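# Material function that produces a camera-distance fade value to mask vehicles out of sight when a passenger aims down sights.
# It offsets the pixel-to-camera distance by scalar "s" and divides by "cull distance", both parameters described as Blueprint-controlled.
# A shading path switch runs the result through DitherTemporalAA_Fortnite by default, while mobile outputs the undithered ratio directly.
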MaterialFunction /Game/Packages/Fortress_SharedMaterials/Base_Material_Functions/BaseCustomizationFunctions/MF_BaseVehicle_ADS_Dithering
graph {nodes [
  n1 = WorldPosition()
  n2 = CameraPositionWS()
  n3 = ScalarParameter() {
   defaultValue=-40
   desc=Controlled by BP to ensure vehicle is masked out of sight when using ADS as passenger.
   group=Vehicle_ADS_Dithering
   parameterName=s
  }
  n4 = ScalarParameter() {
   defaultValue=40
   desc=Controlled by BP to ensure vehicle is masked out of sight when using ADS as passenger.
   group=Vehicle_ADS_Dithering
   parameterName=cull distance
  }
  n5 = Divide(A=Add(A=Distance(A=@n1.XYZ, B=@n2), B=@n3), B=@n4)
  n6 = Constant()
  FunctionOutput(None=ShadingPathSwitch(
    Default=MaterialFunctionCall(Alpha Threshold=@n5, Random=@n6) {materialFunction=&DitherTemporalAA_Fortnite}.Result
    Mobile=@n5
   ))
 ]}
