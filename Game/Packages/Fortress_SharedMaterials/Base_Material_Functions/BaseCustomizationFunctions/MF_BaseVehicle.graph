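# Base vehicle material function that builds shared material attributes from diffuse, normal, SpecularMasks and M textures, and exposes individual channels as outputs.
# SpecularMasks drives specular, metalness and multiplied, clamped roughness, while an ADS dithering function supplies opacity mask.
# M provides AO, GMap and a clear coat/damage mask scaled by opacity for clear coat, with constants replacing it on mobile.
# A lights texture outputs tailpipe, headlight and taillight masks, which a static switch zeroes.
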MaterialFunction /Game/Packages/Fortress_SharedMaterials/Base_Material_Functions/BaseCustomizationFunctions/MF_BaseVehicle
graph {nodes [
  Constant3Vector() {constant=<rgba 0.5 0 0.5 0>}
  Constant()
  n1 = TextureSampleParameter2D() {
   group=Vehicle_Textures
   parameterName=SpecularMasks
   samplerSource=SSM_Wrap_WorldGroupSettings
   samplerType=SAMPLERTYPE_Masks
   sortPriority=1
   texture=&T_BlackMask
  }
  n2 = MaterialFunctionCall(Float3=ShadingPathSwitch(Default=@n1.RGB) {desc=Migrated from FeatureLevelSwitch}) {materialFunction=&BreakOutFloat3Components}
  n3 = Multiply(
   A=FunctionOutput(None=@n2.B) {outputName="Roughness (_S B)", sortPriority=4}
   B=ScalarParameter() {defaultValue=1, parameterName=RoughnessMultiplier}
  )
  n4 = TextureSampleParameter2D() {
   group=Vehicle_Textures
   parameterName=Normals
   samplerSource=SSM_Wrap_WorldGroupSettings
   samplerType=SAMPLERTYPE_Normal
   sortPriority=2
   texture=&FlatNormalBlue
  }
  n5 = ScalarParameter() {defaultValue=1, group=Vehicle_ClearCoat, parameterName=Clear Coat Opacity, sliderMax=1}
  n6 = TextureSampleParameter2D() {
   group=Vehicle_Textures
   parameterName=M
   samplerSource=SSM_Wrap_WorldGroupSettings
   samplerType=SAMPLERTYPE_Masks
   sortPriority=3
   texture=&T_WhiteMask
  }
  n7 = MaterialFunctionCall(Float3=ShadingPathSwitch(Default=@n6.RGB, Mobile=Constant3Vector() {constant=<rgba 1 1 0 0>}) {desc=Migrated from FeatureLevelSwitch}) {materialFunction=&BreakOutFloat3Components}
  n8 = TextureSampleParameter2D() {group=Vehicle_Textures, parameterName=Diffuse, texture=&T_BlacksRGB}
  FunctionOutput(None=MakeMaterialAttributes!(
    BaseColor=FunctionOutput(None=@n8.RGB) {outputName=Diffuse, sortPriority=1}
    Metallic=FunctionOutput(None=@n2.G) {outputName="Metalness (_S G)", sortPriority=3}
    Specular=FunctionOutput(None=@n2.R) {outputName="Specular (_S R)", sortPriority=2}
    Roughness=Clamp(None=@n3)
    Anisotropy
    EmissiveColor
    Opacity=Constant()
    OpacityMask=MaterialFunctionCall() {materialFunction=&/MF_BaseVehicle_ADS_Dithering}.Result
    Normal=FunctionOutput(None=@n4.RGB) {outputName=Normal Map, sortPriority=5}
    Tangent
    WorldPositionOffset
    SubsurfaceColor
    ClearCoat=Multiply(A=@n7.B, B=@n5)
    ClearCoatRoughness=ScalarParameter() {group=Vehicle_ClearCoat, parameterName=ClearCoatRoughness}
    AmbientOcclusion
    Refraction
    CustomizedUVs_0
    CustomizedUVs_1
    CustomizedUVs_2
    CustomizedUVs_3
    CustomizedUVs_4
    CustomizedUVs_5
    CustomizedUVs_6
    CustomizedUVs_7
    PixelDepthOffset
    ShadingModel
    Displacement
   )) {outputName=Material Result, sortPriority=-200}
  FunctionOutput(None=Constant()) {outputName=-------------------, sortPriority=10}
  n9 = TextureSampleParameter2D() {
   group=Vehicle_Textures
   parameterName=VehicleLightsMask
   samplerSource=SSM_Wrap_WorldGroupSettings
   samplerType=SAMPLERTYPE_Masks
   sortPriority=4
   texture=&T_BlackMask
  }
  n10 = MaterialFunctionCall(Float3=StaticSwitchParameter(True=Constant3Vector(), False=@n9.RGB) {parameterName=Disable Vehicle Lights}) {materialFunction=&BreakOutFloat3Components}
  FunctionOutput(None=@n10.R) {outputName="Tail Pipe Mask (_VL R)", sortPriority=21}
  FunctionOutput(None=@n10.G) {outputName="Head Light Mask (_VL G)", sortPriority=22}
  FunctionOutput(None=@n10.B) {outputName="Tail Light Mask (_VL B)", sortPriority=23}
  FunctionOutput(None=@n7.R) {outputName="AO (_M R)", sortPriority=16}
  FunctionOutput(None=Constant()) {outputName=------------------, sortPriority=-19}
  FunctionOutput(None=@n7.G) {outputName="GMap (_M G)", sortPriority=17}
  FunctionOutput(None=@n7.B) {description=Mask for Area that will have Clear Coat and/or visible Damage, outputName="CC (_M B)", sortPriority=18}
 ]}
