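# Material function that samples the CustomizationMask (_CM) texture parameter on an input UV 1 set and exposes it for character or item customization in shared base materials.
# Outputs provide RGB, RGBA and individual channels as numbered masks, GMAP, crevasse and UV islands, the latter being just red.
# A Custom Tex Coord 7 output for Supercharged adds AverageHandPosition to the mask, then multiplies by zero.
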MaterialFunction /Game/Packages/Fortress_SharedMaterials/Base_Material_Functions/BaseCustomizationFunctions/MF_Base_CustomizationMask
graph {nodes [
  n1 = FunctionInput(Preview=TextureCoordinate()) {bUsePreviewValueAsDefault=true, inputName=UV 1 Coordinates, inputType=FunctionInput_Vector2, sortPriority=2}
  n2 = TextureSampleParameter2D(UVs=@n1) {
   group=Customization
   parameterName="CustomizationMask (_CM)"
   samplerSource=SSM_Wrap_WorldGroupSettings
   samplerType=SAMPLERTYPE_Masks
   sortPriority=-400
   texture=&T_BlackMask
  }
  FunctionOutput(None=@n2.G) {outputName="Customization Mask 3 (_CM G)", sortPriority=7}
  FunctionOutput(None=@n2.A) {outputName="Customization GMAP (_CM A)", sortPriority=3}
  FunctionOutput(None=@n2.RGB) {outputName="Customization Mask (_CM RGB)", sortPriority=1}
  FunctionOutput(None=@n2.R) {outputName="Customization Mask 1 (_CM R)", sortPriority=6}
  FunctionOutput(None=@n2.B) {outputName="Customization Mask 2 (_CM B)", sortPriority=7}
  n3 = VectorParameter() {group=Super Charge, parameterName=AverageHandPosition, defaultValue=<rgba 0 0 0 1>}
  FunctionOutput(None=Multiply(A=Add(A=@n3.RGB, B=@n2.RGB)) {constB=0}) {description="Passes Through the _CM Mask, used for Supercharged Effect", outputName=Custom Tex Coord 7, sortPriority=-4}
  FunctionOutput(None=Max(A=@n2.R, B=@n2.R)) {outputName=Customization Mask UV Islands, sortPriority=9}
  FunctionOutput(None=@n2.G) {outputName="Customization Crevasse (_CM G)", sortPriority=3}
  FunctionOutput(None=Constant()) {outputName=-------------------}
  FunctionOutput(None=@n2.RGBA) {outputName="Customization Mask (_CM RGBA)", sortPriority=1}
  FunctionOutput(None=Constant()) {outputName=--------------------, sortPriority=5}
  Multiply(A=@n1) {constB=0.5}
  TextureSampleParameter2D() {
   group=Customization
   parameterName="CustomizationMask (_CM)"
   samplerSource=SSM_Wrap_WorldGroupSettings
   samplerType=SAMPLERTYPE_Masks
   sortPriority=-400
   texture=&T_BlackMask
  }
 ]}
