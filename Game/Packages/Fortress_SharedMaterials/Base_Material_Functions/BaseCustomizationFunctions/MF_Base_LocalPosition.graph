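# Material function that exposes pre-skinned local-space mesh data to the pixel shader for Fortress shared materials.
# It outputs local position, local normal and the half extents of the local bounds, all taken before skinning.
# Each value is read in the vertex stage and passed through its own VertexInterpolator, making these vertex-only inputs usable per pixel.
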MaterialFunction /Game/Packages/Fortress_SharedMaterials/Base_Material_Functions/BaseCustomizationFunctions/MF_Base_LocalPosition
graph {nodes [
  FunctionOutput(None=VertexInterpolator(VS=PreSkinnedLocalBounds().Half Extents).PS) {outputName=Pre-Skinned Local Bounds Pixel Shader, sortPriority=5}
  FunctionOutput(None=VertexInterpolator(VS=PreSkinnedPosition()).PS) {outputName=Pre-Skinned Local Position Pixel Shader, sortPriority=2}
  FunctionOutput(None=VertexInterpolator(VS=PreSkinnedNormal()).PS) {outputName=Pre-Skinned Local Normal Pixel Shader, sortPriority=5}
 ]}
