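# Master material function for Fortress vehicles, bundling base surface and customization data behind one set of outputs.
# MF_BaseVehicle supplies material result, normal map, clear coat/damage mask, and tail pipe, headlight and tail light masks.
# MF_Base_CustomizationMask supplies _CM masks, crevasse and gradient map channel, while MF_Base_LocalPosition exposes pre-skinned local position, normal and bounds.
# Constant-fed outputs act as separators, and Customization Mask 3 is unconnected.
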MaterialFunction /Game/Packages/Fortress_SharedMaterials/Base_Material_Functions/BaseCustomizationFunctions/MF_FN_Vehicle_Master.MF_FN_Vehicle_MASTER
graph {nodes [
  n1 = MaterialFunctionCall() {materialFunction=&/MF_BaseVehicle}
  FunctionOutput(None=@n1.Material Result) {outputName=Material Result, sortPriority=-200}
  FunctionOutput(None=Constant()) {outputName=-------------------, sortPriority=23}
  FunctionOutput(None=@n1."Tail Pipe Mask (_VL R)") {outputName="Tail Pipe Mask (_VL R)", sortPriority=21}
  n2 = MaterialFunctionCall!(UV 1 Coordinates) {materialFunction=&/MF_Base_CustomizationMask}
  FunctionOutput(None=@n2."Customization Mask 1 (_CM R)") {outputName="Customization Mask 1 (_CM B Top Left)", sortPriority=24}
  FunctionOutput(None=@n1."Tail Light Mask (_VL B)") {outputName="Tail Light Mask (_VL B)", sortPriority=23}
  FunctionOutput(None=@n2."Customization GMAP (_CM A)") {outputName=Customization Gmap, sortPriority=21}
  FunctionOutput(None=@n2."Customization Mask 2 (_CM B)") {outputName="Customization Mask 2 (_CM B Top Right)", sortPriority=25}
  n3 = MaterialFunctionCall() {materialFunction=&/MF_Base_LocalPosition}
  FunctionOutput(None=@n3.Pre-Skinned Local Normal Pixel Shader) {outputName=Pre-Skinned Local Normal Pixel Shader, sortPriority=42}
  FunctionOutput(None=@n2.Customization Mask UV Islands) {outputName="Customization Mask UV Islands (_CM B Bottom Right)", sortPriority=22}
  FunctionOutput(None=Constant()) {outputName=------------------, sortPriority=-19}
  FunctionOutput(None=@n2."Customization Crevasse (_CM G)") {outputName="Customization Crevasse (_CM G)", sortPriority=21}
  FunctionOutput(None=Constant()) {outputName=--------------------, sortPriority=40}
  FunctionOutput(None=@n2."Customization Mask (_CM RGBA)") {outputName="Customization Mask (_CM RGBA)", sortPriority=31}
  FunctionOutput(None=@n3.Pre-Skinned Local Position Pixel Shader) {outputName=Pre-Skinned Local Position Pixel Shader, sortPriority=41}
  FunctionOutput() {outputName="Customization Mask 3 (_CM B Bottom Left)", sortPriority=26}
  FunctionOutput(None=@n1.Normal Map) {outputName=Normal Map, sortPriority=15}
  FunctionOutput(None=@n1."Head Light Mask (_VL G)") {outputName="Head Light Mask (_VL G)", sortPriority=22}
  FunctionOutput(None=@n3.Pre-Skinned Local Bounds Pixel Shader) {outputName=Pre-Skinned Local Bounds Pixel Shader, sortPriority=43}
  FunctionOutput(None=@n1."CC (_M B)") {description=Mask for Area that will have Clear Coat and/or visible Damage, outputName="CC/Damage Mask (_M B)", sortPriority=18}
 ]}
