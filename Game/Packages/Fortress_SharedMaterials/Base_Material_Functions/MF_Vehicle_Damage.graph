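# Material function that layers scratch damage onto vehicle materials, driven by a clear coat/damage mask and a Damage parameter.
# The clamped product of mask and powered Damage blends a tiled damage normal into the incoming normal.
# Scratch areas come from the normal map's squared inverse blue channel, avoiding a separate scratch texture.
# They tint base color, override roughness to one and, with the broader damage mask, raise clear coat roughness.
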MaterialFunction /Game/Packages/Fortress_SharedMaterials/Base_Material_Functions/MF_Vehicle_Damage
graph {nodes [
  n1 = FunctionInput() {
   bUsePreviewValueAsDefault=true
   inputName="Clear Coat/Damage Mask (_M B)"
   inputType=FunctionInput_Scalar
   sortPriority=1
   previewValue=<1 1 1 1>
  }
  n2 = ScalarParameter() {group=Vehicle_Damage, parameterName=Damage}
  n3 = Clamp(None=Multiply(A=@n1, B=Power(Base=@n2)))
  n4 = FunctionInput() {bUsePreviewValueAsDefault=true, inputType=FunctionInput_MaterialAttributes, previewValue=<0.5 0.5 0.5 1>}
  n5 = VectorParameter() {group=Vehicle_Damage, parameterName=DamageTextureTile, defaultValue=<rgba 1.5 1.5 0 0>}
  n6 = TextureSample(UVs=Multiply(A=TextureCoordinate() {uTiling=0.7, vTiling=0.7}, B=ComponentMask(None=@n5.RGB) {g=true, r=true})) {samplerType=SAMPLERTYPE_Normal, texture=/Game/Global/Textures/Default/Tileable_Patterns/CarDamageNormal}
  n7 = VectorParameter() {group=Vehicle_Damage, parameterName=DamageNormalMult, defaultValue=<rgba 1.6 1.6 0 1>}
  n8 = MaterialFunctionCall!(
   NormalMask=@n3
   Material=@n4
   Normal=Add(
    A=MaterialFunctionCall!(Material=@n4) {materialFunction=/Engine/Functions/MaterialLayerFunctions/MatLayerBlend_BreakNormal}.Normal
    B=Multiply(A=@n6.RGB, B=@n7.RGB)
   )
  ) {materialFunction=/Engine/Functions/MaterialLayerFunctions/MatLayerBlend_NormalBlend}
  n9 = Multiply(
   A=MaterialFunctionCall!(Material=@n4) {materialFunction=/Engine/Functions/MaterialLayerFunctions/MatLayerBlend_BreakBaseColor}.BaseColor
   B=VectorParameter() {group=Vehicle_Damage, parameterName=DamageScratchColor, defaultValue=<rgba 0.2 0.1666 0.1289 0>}.RGB
  )
  n10 = OneMinus(None=@n6.B)
  n11 = Multiply(A=@n10, B=@n10)
  n12 = Multiply(A=@n3, B=Clamp(None=Multiply(A=@n11) {constB=10}))
  n13 = BreakMaterialAttributes!(Attr=MaterialFunctionCall!(
    Material=MaterialFunctionCall!(Material=@n8.Blended Material, NewBaseColor=@n9, Mask=@n12) {materialFunction=/Engine/Functions/MaterialLayerFunctions/MatLayerBlend_MultiplyBaseColor}.Blended Material
    NewRoughness=Constant() {r=1}
    Mask=@n12
   ) {materialFunction=/Engine/Functions/MaterialLayerFunctions/MatLayerBlend_RoughnessOverride}.Blended Material)
  n14 = Add(A=@n12, B=Multiply(A=@n3) {constB=0.15})
  FunctionOutput(None=MakeMaterialAttributes!(
    BaseColor=@n13.BaseColor
    Metallic=@n13.BaseColor
    Specular=@n13.BaseColor
    Roughness=@n13.BaseColor
    Anisotropy
    EmissiveColor=@n13.BaseColor
    Opacity=@n13.BaseColor
    OpacityMask=@n13.BaseColor
    Normal=@n13.BaseColor
    Tangent
    WorldPositionOffset=@n13.BaseColor
    SubsurfaceColor=@n13.BaseColor
    ClearCoat=@n13.BaseColor
    ClearCoatRoughness=Add(A=@n13.ClearCoatRoughness, B=@n14)
    AmbientOcclusion
    Refraction
    CustomizedUVs_0
    CustomizedUVs_1
    CustomizedUVs_2
    CustomizedUVs_3
    CustomizedUVs_4
    CustomizedUVs_5
    CustomizedUVs_6
    CustomizedUVs_7
    PixelDepthOffset
    ShadingModel
    Displacement
   ))
 ]}
