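# Material function that optionally applies vehicle damage shading to incoming material attributes for Fortress vehicle materials.
# It passes the attributes and a clear coat/damage mask scalar, both with preview defaults, into MF_Vehicle_Damage.
# A static switch parameter, "Vehicle Body Shows Damage" in the Vehicle_Damage group, defaults on and selects the damaged result or the untouched input at compile time.
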MaterialFunction /Game/Packages/Fortress_SharedMaterials/Base_Material_Functions/BaseCustomizationFunctions/MF_BaseVehicle_Damage
graph {nodes [
  n1 = FunctionInput() {bUsePreviewValueAsDefault=true, inputName=Material, inputType=FunctionInput_MaterialAttributes, previewValue=<0.5 0.5 0.5 1>}
  n2 = FunctionInput() {
   bUsePreviewValueAsDefault=true
   inputName="Clear Coat/Damage Mask (_M B)"
   inputType=FunctionInput_Scalar
   sortPriority=1
   previewValue=<1 1 1 1>
  }
  FunctionOutput(None=StaticSwitchParameter(
    True=MaterialFunctionCall!(In=@n1, "Clear Coat/Damage Mask (_M B)"=@n2) {materialFunction=/Game/Packages/Fortress_SharedMaterials/Base_Material_Functions/MF_Vehicle_Damage}.Result
    False=@n1
   ) {defaultValue=true, group=Vehicle_Damage, parameterName=Vehicle Body Shows Damage})
 ]}
