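# Material function that adds a fading overheat glow to the emissive of vehicle weapons.
# Glow strength is WeaponGlow minus the normalized time elapsed since GlowTime over GlowDuration, so the material animates the cooldown from Time.
# That strength, shaped by Weapon Alpha with adjustable squared contrast, indexes a black-body gradient texture, whose color is scaled by the glow and added to Emissive.
# Weapon Alpha then blends the modified attributes over the input Material.
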MaterialFunction /Game/Packages/Fortress_SharedMaterials/Base_Material_Functions/BaseCustomizationFunctions/MF_BaseVehicle_WeaponOverheat
graph {nodes [
  n1 = FunctionInput() {inputName=Material, inputType=FunctionInput_MaterialAttributes}
  n2 = GetMaterialAttributes(None=@n1) {attributeGetTypes [#1]}
  n3 = Divide(
   A=Subtract(A=Time(), B=ScalarParameter() {defaultValue=5.076944363720802e+18, group=Vehicle_Overheat, parameterName=GlowTime})
   B=ScalarParameter() {defaultValue=3, group=Vehicle_Overheat, parameterName=GlowDuration}
  )
  n4 = Saturate(None=Subtract(A=ScalarParameter() {group=Vehicle_Overheat, parameterName=WeaponGlow}, B=Saturate(None=@n3)))
  n5 = FunctionInput() {inputName=Weapon Alpha, inputType=FunctionInput_Scalar, sortPriority=1}
  n6 = LinearInterpolate(
   A=Constant()
   B=Multiply(A=@n4, B=Constant() {r=0.95})
   Alpha=LinearInterpolate(A=@n5, B=Multiply(A=@n5, B=@n5), Alpha=ScalarParameter() {defaultValue=1, group=Vehicle_Overheat, parameterName=WeaponAlphaContrast})
  )
  n7 = Add(
   A=@n2.EmissiveColor
   B=Multiply(
    A=@n4
    B=TextureSample(UVs=@n6) {desc=Black Body Gradient, texture=/Game/Global/Textures/Default/Masks_and_Gradients/WeaponGradients/T_BlackBody_Gradient_2}.RGB
   )
  )
  FunctionOutput(None=BlendMaterialAttributes(
    A=@n1
    B=SetMaterialAttributes!(MaterialAttributes=@n2.MaterialAttributes, EmissiveColor=@n7) {attributeSetTypes [#1]}
    Alpha=@n5
   ))
 ]}
