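# Material function that adds light and exhaust effects to vehicle material attributes, using head light, tail light and tail pipe masks from _VL G, B and R.
# Automatic night-time headlights feed two parallel branches, TurnSignals and WeaponOverheat, both masked by the blue tail light channel.
# The UseOverheatGlow scalar blends between those branches, a noted workaround for overheat glow on vehicle wraps.
# The blended result passes through the tail pipe function to the output.
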MaterialFunction /Game/Packages/Fortress_SharedMaterials/Base_Material_Functions/BaseCustomizationFunctions/MF_BaseVehicle_Lights
graph {nodes [
  n1 = MaterialFunctionCall!(
   Alpha=FunctionInput() {bUsePreviewValueAsDefault=true, inputName="Head Light Mask (_VL G)", inputType=FunctionInput_Scalar, sortPriority=2}
   In=FunctionInput() {bUsePreviewValueAsDefault=true, inputName=Material, inputType=FunctionInput_MaterialAttributes}
  ) {materialFunction=/Game/Weapons/WeaponSkins/MaterialFunction/NightTime_AutomaticHeadlights}
  n2 = FunctionInput() {bUsePreviewValueAsDefault=true, inputName="Tail Light Mask (_VL B)", inputType=FunctionInput_Scalar, sortPriority=3}
  n3 = MaterialFunctionCall!(In=@n1.Result, CenterLine, Alpha=@n2) {materialFunction=/Game/Weapons/WeaponSkins/MaterialFunction/TurnSignals}
  n4 = MaterialFunctionCall!(Material=@n1.Result, Weapon Alpha=@n2) {
   desc=Use blue mask channel for weapon glow
   materialFunction=&/MF_BaseVehicle_WeaponOverheat
  }
  n5 = ScalarParameter() {parameterName=UseOverheatGlow}
  FunctionOutput(None=MaterialFunctionCall!(
    Material=BlendMaterialAttributes(A=@n3.Result, B=@n4.Result, Alpha=@n5) {desc=Workaround for making overheat glow work on vehicle wraps}
    Tail Pipe Alpha=FunctionInput() {bUsePreviewValueAsDefault=true, inputName="Tail Pipe Mask (_VL R)", inputType=FunctionInput_Scalar, sortPriority=1}
   ) {materialFunction=&/MF_BaseVehicle_TailPipe}.Result)
 ]}
