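# Material function that provides a tintable emissive layer confined to a UV rectangle, with a second identical output labeled as a customization passthrough.
# It crops EmissiveTexture via TextureCropping, taking the top-left corner from a vector parameter's RG and the bottom-right from B duplicated, despite the BA name.
# The masked result is multiplied by EmissiveMultiplier, optionally swapped for a mobile-specific multiplier through shading-path switching.
# A static bool replaces both outputs with zero to drop the texture.
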MaterialFunction /Game/Packages/Fortress_SharedMaterials/Base_Material_Functions/BaseCustomizationFunctions/MF_Base_CroppedEmissive
graph {nodes [
  n1 = Constant()
  n2 = VectorParameter() {group=Emissive, parameterName=Emissive Texture UVs RG_TopLeft BA_BottomRight, defaultValue=<rgba 0 0 1 1>}
  n3 = VectorParameter() {group=Emissive, parameterName=EmissiveMultiplier, defaultValue=<rgba 1 1 1 1>}
  n4 = VectorParameter() {group=Emissive, parameterName=EmissiveMultiplier_Mobile, defaultValue=<rgba 1 1 1 1>}
  n5 = Multiply(
   A=MaterialFunctionCall!(
    UVs
    TextureIn=TextureObjectParameter() {group=Emissive, parameterName=EmissiveTexture, texture=&T_BlacksRGB}
    UpperLeftCorner=ComponentMask(None=@n2.RGB) {g=true, r=true}
    LowerRightCorner=AppendVector(A=@n2.B, B=@n2.B)
    ExportFloat4
   ) {materialFunction=/Engine/Functions/Engine_MaterialFunctions01/Texturing/TextureCropping}.CroppedMasked
   B=StaticSwitchParameter(True=ShadingPathSwitch(Default=@n3.RGB, Deferred=@n3.RGB, Mobile=@n4.RGB) {desc=Migrated from FeatureLevelSwitch}, False=@n3.RGB) {parameterName=OverrideEmissiveMultonMobile}
  )
  n6 = StaticBoolParameter() {group=Emissive, parameterName=Get Rid of Emissive Texture}
  FunctionOutput(None=StaticSwitch(True=@n1, False=@n5, Value=@n6))
  FunctionOutput(None=StaticSwitch(True=@n1, False=@n5, Value=@n6)) {description=Passthrough for Customization, outputName=Cropped Emissive Texture, sortPriority=2}
 ]}
