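# Material function that builds the emissive layer for customizable vehicle materials, including tail pipe, headlight and tail light effects.
# It adds MF_Base_CroppedEmissive to the incoming EmissiveColor, then runs the attributes through WeaponEmissiveTOD, a function borrowed from weapon skins.
# MF_BaseVehicle_Lights then applies the three scalar light masks, labeled as _VL R, G and B channels.
# A second output passes the cropped emissive texture through for customization.
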MaterialFunction /Game/Packages/Fortress_SharedMaterials/Base_Material_Functions/BaseCustomizationFunctions/MF_BaseVehicle_Emissive
graph {nodes [
  n1 = GetMaterialAttributes(None=FunctionInput() {bUsePreviewValueAsDefault=true, inputName=Material, inputType=FunctionInput_MaterialAttributes}) {attributeGetTypes [#1]}
  n2 = MaterialFunctionCall() {materialFunction=&/MF_Base_CroppedEmissive}
  n3 = SetMaterialAttributes!(MaterialAttributes=@n1.MaterialAttributes, EmissiveColor=Add(A=@n1.EmissiveColor, B=@n2.Result)) {attributeSetTypes [#1]}
  FunctionOutput(None=MaterialFunctionCall!(
    Material=MaterialFunctionCall!(In=@n3) {materialFunction=/Game/Weapons/WeaponSkins/MaterialFunction/WeaponEmissiveTOD}.Result
    "Tail Pipe Mask (_VL R)"=FunctionInput() {bUsePreviewValueAsDefault=true, inputName="Tail Pipe Mask (_VL R)", inputType=FunctionInput_Scalar, sortPriority=1}
    "Head Light Mask (_VL G)"=FunctionInput() {bUsePreviewValueAsDefault=true, inputName="Head Light Mask (_VL G)", inputType=FunctionInput_Scalar, sortPriority=2}
    "Tail Light Mask (_VL B)"=FunctionInput() {bUsePreviewValueAsDefault=true, inputName="Tail Light Mask (_VL B)", inputType=FunctionInput_Scalar, sortPriority=3}
   ) {materialFunction=&/MF_BaseVehicle_Lights}.Result)
  FunctionOutput(None=@n2.Cropped Emissive Texture) {description=Passthrough for Customization, outputName=Cropped Emissive Texture, sortPriority=2}
 ]}
