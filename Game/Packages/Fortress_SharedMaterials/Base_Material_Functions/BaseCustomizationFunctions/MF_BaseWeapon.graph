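# Material function providing the base PBR surface layer for weapons, sampling Diffuse, SpecularMasks and Normals texture parameters on one shared UV input.
# Specular mask channels drive scaled specular (R), metalness (G) and roughness remapped between min/max parameters (B).
# An isClearCoat static switch blends metalness toward PaintMetallic using mask alpha.
# That alpha is also output as the clear coat/damage mask, though labelled "_M B".
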MaterialFunction /Game/Packages/Fortress_SharedMaterials/Base_Material_Functions/BaseCustomizationFunctions/MF_BaseWeapon
graph {nodes [
  n1 = FunctionInput(Preview=TextureCoordinate()) {bUsePreviewValueAsDefault=true, inputName=UV 1 Coordinates, inputType=FunctionInput_Vector2, sortPriority=2}
  n2 = TextureSampleParameter2D(UVs=@n1) {group=Weapon_Textures, parameterName=Diffuse, texture=/Game/Weapons/FORT_Shotguns/Textures/T_HammerShotGun_D}
  FunctionOutput(None=@n2.A) {outputName=Diffuse Alpha, sortPriority=1}
  n3 = TextureSampleParameter2D(UVs=@n1) {
   constCoordinate=1
   group=Weapon_Textures
   parameterName=SpecularMasks
   samplerType=SAMPLERTYPE_Masks
   sortPriority=1
   texture=/Game/Weapons/FORT_Shotguns/Textures/T_HammerShotGun_S
  }
  n4 = LinearInterpolate(
   A=ScalarParameter() {group=Roughness, parameterName=SpecRoughnessMin}
   B=ScalarParameter() {defaultValue=1, group=Roughness, parameterName=SpecRoughnessMax}
   Alpha=@n3.B
  )
  FunctionOutput(None=@n4) {outputName="Roughness (_S B)", sortPriority=4}
  n5 = TextureSampleParameter2D(UVs=@n1) {
   constCoordinate=2
   group=Weapon_Textures
   parameterName=Normals
   samplerType=SAMPLERTYPE_Normal
   sortPriority=2
   texture=/Game/Weapons/FORT_Shotguns/Textures/T_HammerShotGun_N
  }
  FunctionOutput(None=@n5.RGB) {outputName=Normal Map, sortPriority=5}
  n6 = Multiply(A=@n3.R, B=ScalarParameter() {defaultValue=1, group=Specular, parameterName=SpecularLevel})
  FunctionOutput(None=@n6) {outputName="Specular (_S R)", sortPriority=2}
  FunctionOutput(None=@n2.RGB) {outputName=Diffuse, sortPriority=1}
  n7 = ScalarParameter() {defaultValue=1, group=ClearCoat, parameterName=PaintMetallic}
  n8 = StaticSwitchParameter(True=LinearInterpolate(A=@n3.G, B=Multiply(A=@n7, B=@n3.A), Alpha=@n3.A), False=@n3.G) {parameterName=isClearCoat}
  FunctionOutput(None=@n8) {outputName="Metalness (_S G)", sortPriority=3}
  FunctionOutput(None=@n3.A) {description=Mask for Area that will have Clear Coat and/or visible Damage, outputName="CC (_M B)", sortPriority=9}
  Reroute(None=@n2.RGB)
  Reroute(None=@n6)
  Reroute(None=@n5.RGB)
  Reroute(None=@n4)
  Reroute(None=@n8)
 ]}
