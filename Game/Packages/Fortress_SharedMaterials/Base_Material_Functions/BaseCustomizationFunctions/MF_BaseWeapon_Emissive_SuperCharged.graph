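# Material function that adds the super-charged glow to a weapon's emissive, reusing the character shader's effect.
# It calls SuperChargedGlow with the SuperChargedEffect and SuperChargedFlickerBias scalar parameters from the Super Charge group, then adds the result to the incoming Emissive.
# UseDistanceCalc is statically off, and the glow function's AverageHandPosition and Emissive inputs are left unconnected.
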MaterialFunction /Game/Packages/Fortress_SharedMaterials/Base_Material_Functions/BaseCustomizationFunctions/MF_BaseWeapon_Emissive_SuperCharged
graph {nodes [
  n1 = ScalarParameter() {defaultValue=1, group=Super Charge, parameterName=SuperChargedFlickerBias}
  n2 = ScalarParameter() {group=Super Charge, parameterName=SuperChargedEffect}
  n3 = StaticBool()
  FunctionOutput(None=Add(
    A=FunctionInput() {bUsePreviewValueAsDefault=true, inputName=Emissive}
    B=MaterialFunctionCall!(AverageHandPosition, SuperChargedFlickerBias=@n1, SuperChargedEffect=@n2, UseDistanceCalc=@n3, Emissive) {materialFunction=/Game/Characters/Fortress_CharacterShader/SuperChargedGlow}.Result
   ))
 ]}
