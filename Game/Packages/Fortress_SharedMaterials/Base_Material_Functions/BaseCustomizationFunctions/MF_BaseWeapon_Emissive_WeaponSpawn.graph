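# Material function that adds a fresnel-tinted glow to the back faces of two-sided weapon materials, labeled as a weapon spawn-in emissive.
# A clamped, negated TwoSidedSign masks to back faces, multiplying a lerp between inner and outer color parameters by absolute fresnel.
# The inner color's alpha doubles as the fresnel power.
# The result adds to the incoming Emissive, unless a static switch disables it by substituting zero.
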MaterialFunction /Game/Packages/Fortress_SharedMaterials/Base_Material_Functions/BaseCustomizationFunctions/MF_BaseWeapon_Emissive_WeaponSpawn
graph {nodes [
  Constant()
  n1 = Constant3Vector()
  n2 = TwoSidedSign()
  n3 = VectorParameter() {desc=Weapon Spawn-in Emissive Fresnel Inner Color, group=Weapon_Spawn, parameterName=colcol1, defaultValue=<rgba 6.5583 9.5135 20 0.2>}
  n4 = VectorParameter() {
   desc=Weapon Spawn-in Emissive Fresnel Outer Color
   group=Weapon_Spawn
   parameterName=colcol2
   sortPriority=1
   defaultValue=<rgba 1.1035 1.5954 6 1>
  }
  n5 = MaterialFunctionCall(Power=@n3.A, Clamp Fresnel Dot Product=StaticBool()) {materialFunction=&Fresnel_Function}
  n6 = Multiply(A=Clamp(None=Multiply(A=@n2) {constB=-1}), B=LinearInterpolate(A=@n3.RGB, B=@n4.RGB, Alpha=Abs(None=@n5.Result)))
  FunctionOutput(None=Add(
    A=FunctionInput() {bUsePreviewValueAsDefault=true, inputName=Emissive}
    B=StaticSwitchParameter(True=@n1, False=@n6) {group=Emissive, parameterName=Disable InteriorEmissiveForTwoSided}
   ))
 ]}
