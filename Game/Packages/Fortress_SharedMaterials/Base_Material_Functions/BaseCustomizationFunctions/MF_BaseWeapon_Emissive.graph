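# Material function that builds the emissive for base weapon materials.
# MF_Base_CroppedEmissive supplies the base glow, and its cropped emissive texture also passes through as a second output for customization.
# The WeaponSpawn emissive effect layers on only for non-mobile shading at Default or High quality; Mobile, Low and Medium use the base glow.
# SuperCharged and WeaponEmissiveTOD calls are present but feed no output.
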MaterialFunction /Game/Packages/Fortress_SharedMaterials/Base_Material_Functions/BaseCustomizationFunctions/MF_BaseWeapon_Emissive
graph {nodes [
  n1 = MaterialFunctionCall() {materialFunction=&/MF_Base_CroppedEmissive}
  n2 = MaterialFunctionCall!(Emissive=@n1.Result) {materialFunction=&/MF_BaseWeapon_Emissive_WeaponSpawn}
  n3 = ShadingPathSwitch(Default=@n2.Result, Mobile=@n1.Result) {desc=Migrated from FeatureLevelSwitch}
  FunctionOutput(None=QualitySwitch(Default=@n3, Low=@n1.Result, High=@n3, Medium=@n1.Result))
  FunctionOutput(None=@n1.Cropped Emissive Texture) {description=Passthrough for Customization, outputName=Cropped Emissive Texture, sortPriority=2}
  MaterialFunctionCall!(Emissive=@n1.Result) {materialFunction=&/MF_BaseWeapon_Emissive_SuperCharged}
  MaterialFunctionCall!(In) {materialFunction=/Game/Weapons/WeaponSkins/MaterialFunction/WeaponEmissiveTOD}
 ]}
