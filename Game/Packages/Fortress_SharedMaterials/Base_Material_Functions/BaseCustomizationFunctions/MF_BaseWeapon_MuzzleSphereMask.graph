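# Material function for weapon materials that builds a soft spherical mask around the muzzle in local space.
# A SphereMask with zero hardness compares the Muzzle_LocalPosition vector parameter with the Pre-Skinned Local Position input.
# Its radius comes from the MuzzleFalloff input, which defaults to a scalar parameter of the same name.
# The function outputs both the muzzle position and the mask.
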MaterialFunction /Game/Packages/Fortress_SharedMaterials/Base_Material_Functions/BaseCustomizationFunctions/MF_BaseWeapon_MuzzleSphereMask
graph {nodes [
  n1 = VectorParameter() {group=Weapon_Muzzle, parameterName=Muzzle_LocalPosition, defaultValue=<rgba 0 0 0 1>}
  FunctionOutput(None=@n1.RGB) {outputName=Muzzle_LocalPosition, sortPriority=43}
  n2 = ScalarParameter() {defaultValue=50, group=Weapon_Muzzle, parameterName=MuzzleFalloff}
  FunctionOutput(None=SphereMask(
    A=@n1.RGB
    B=FunctionInput() {bUsePreviewValueAsDefault=true, inputName=Pre-Skinned Local Position, sortPriority=41}
    Radius=FunctionInput(Preview=@n2) {bUsePreviewValueAsDefault=true, inputName=MuzzleFalloff, inputType=FunctionInput_Scalar, sortPriority=4}
   ) {attenuationRadius=50, hardnessPercent=0}) {outputName=Muzzle_SphereMask, sortPriority=44}
 ]}
