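# MF_BaseWeapon_RandomFloat is a shared base-customization material function that supplies a random float value to weapon materials.
# It wraps a single scalar parameter, RandomFloat, in the Code parameter group and passes it straight to the randomFloat output.
# The function does no randomization itself; the group name suggests game code sets the value.
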MaterialFunction /Game/Packages/Fortress_SharedMaterials/Base_Material_Functions/BaseCustomizationFunctions/MF_BaseWeapon_RandomFloat
graph {nodes [FunctionOutput(None=ScalarParameter() {group=Code, parameterName=RandomFloat}) {outputName=randomFloat, sortPriority=48}]}
