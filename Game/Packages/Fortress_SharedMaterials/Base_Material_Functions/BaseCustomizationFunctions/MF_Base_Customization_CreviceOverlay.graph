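# Material function for weapon or cosmetic customization that overlays crevice detail on Base Color and Specular within the sum of two customization masks.
# The original path scales color by a powered crevice input and dims specular by its powered inverse.
# The V2.0 path splits the CM green channel at mid-grey into peaks and pits, each with separate strength and falloff.
# Static switches select the version or pass attributes through unchanged.
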MaterialFunction /Game/Packages/Fortress_SharedMaterials/Base_Material_Functions/BaseCustomizationFunctions/MF_Base_Customization_CreviceOverlay
graph {nodes [
  n1 = NamedRerouteUsage()
  n2 = Constant3Vector() {constant=<rgba 0.18 0.18 0.18 0>}
  n3 = FunctionInput(Preview=SetMaterialAttributes!(MaterialAttributes, BaseColor=@n2, Specular=@n2) {attributeSetTypes [#1, #2]}) {inputType=FunctionInput_MaterialAttributes}
  n4 = GetMaterialAttributes(None=@n3) {attributeGetTypes [#1, #2]}
  n5 = TextureSample() {samplerType=SAMPLERTYPE_Masks, texture=/Game/Weapons/FORT_Rifles/Textures/T_Scar_CM}
  n6 = FunctionInput(Preview=@n5.G) {bUsePreviewValueAsDefault=true, inputName=Customization Crevice, inputType=FunctionInput_Scalar, sortPriority=1}
  n7 = Power(
   Base=Multiply(A=@n6, B=ScalarParameter() {defaultValue=1, group=Crevice_Overlay, parameterName=CreviceBlendBlance, sortPriority=-60}) {constB=1.2}
   Exp=ScalarParameter() {defaultValue=1, group=Crevice_Overlay, parameterName=CreviceBlendINT, sortPriority=-60}
  )
  n8 = Saturate(None=Multiply(A=@n4.BaseColor, B=Add(A=@n7) {constB=0.5}))
  n9 = ScalarParameter() {defaultValue=1, group=Crevice_Overlay, parameterName=CreviceBlendBascolor, sortPriority=-60}
  n10 = OneMinus(None=@n6)
  n11 = ScalarParameter() {defaultValue=1, group=Crevice_Overlay, parameterName=SpecPower, sortPriority=-60}
  n12 = Saturate(None=Power(
    Base=Multiply(A=@n10, B=@n11)
    Exp=ScalarParameter() {defaultValue=4, group=Crevice_Overlay, parameterName=SpecPowerBlendINT, sortPriority=-60}
   ) {constExponent=4})
  n13 = Saturate(None=Min(A=@n4.Specular, B=Multiply(A=@n4.Specular, B=@n12)))
  n14 = ScalarParameter() {defaultValue=1, group=Crevice_Overlay, parameterName=CreviceBlendSpecular, sortPriority=-60}
  n15 = FunctionInput() {bUsePreviewValueAsDefault=true, inputName="Customization Mask 1 (_CM B Top Left)", inputType=FunctionInput_Scalar, sortPriority=24}
  n16 = FunctionInput() {bUsePreviewValueAsDefault=true, inputName="Customization Mask 2 (_CM B Top Right)", inputType=FunctionInput_Scalar, sortPriority=25}
  n17 = Saturate(None=Add(A=@n15, B=@n16))
  n18 = BlendMaterialAttributes(
   A=@n3
   B=SetMaterialAttributes(
    MaterialAttributes=@n4.MaterialAttributes
    BaseColor=LinearInterpolate(A=@n4.BaseColor, B=@n8, Alpha=@n9)
    Specular=LinearInterpolate(A=@n4.Specular, B=@n13, Alpha=@n14)
   ) {attributeSetTypes [#1, #2]}
   Alpha=@n17
  )
  FunctionOutput(None=StaticSwitchParameter(True=StaticSwitchParameter(True=@n1, False=@n18) {group=Crevice_Overlay, parameterName=CreviceV2.0, sortPriority=-498}, False=@n3) {group=Customization, parameterName=Overlay Crevice Texture, sortPriority=-498})
  n19 = MaterialFunctionCall(Float4=FunctionInput(Preview=@n5.RGBA) {bUsePreviewValueAsDefault=true, inputName=Customization CM, inputType=FunctionInput_Vector4, sortPriority=1}) {materialFunction=&BreakOutFloat4Components}
  Min(A=@n19.G) {constB=0.5}
  n20 = GetMaterialAttributes(None=NamedRerouteUsage()) {attributeGetTypes [#1, #2]}
  n21 = VectorParameter() {
   group=Crevice_Overlay
   parameterName=CreviceBlendBaseColor
   channelNames {a=Peakes Falloff, b=Pits Falloff, g=Peaks, r=Pits}
   defaultValue=<rgba -1 1 1 1>
  }
  n22 = Add(A=@n19.G) {constB=-0.5}
  n23 = Saturate(None=@n22)
  n24 = Multiply(A=@n20.BaseColor, B=Multiply(A=@n21.Peaks, B=Power(Base=@n23, Exp=@n21.Peakes Falloff)))
  n25 = Saturate(None=Multiply(A=@n22) {constB=-1})
  n26 = Multiply(A=@n21.Pits, B=Power(Base=@n25, Exp=@n21.Pits Falloff))
  n27 = Saturate(None=Add(A=Add(A=@n20.BaseColor, B=@n24), B=Multiply(A=@n20.BaseColor, B=@n26)))
  n28 = VectorParameter() {
   group=Crevice_Overlay
   parameterName=CreviceBlendSpec
   channelNames {a=Peakes_Falloff, b=Pits Falloff, g=Peaks, r=Pits}
   defaultValue=<rgba -1 1 1 1>
  }
  n29 = Multiply(A=@n20.Specular, B=Multiply(A=@n28.Peaks, B=Power(Base=@n23, Exp=@n28.Peakes_Falloff)))
  n30 = Multiply(A=@n28.Pits, B=Power(Base=@n25, Exp=@n28.Pits Falloff))
  n31 = Saturate(None=Add(A=Add(A=@n20.Specular, B=@n29), B=Multiply(A=@n20.Specular, B=@n30)))
  NamedRerouteDeclaration(None=BlendMaterialAttributes(
    A=@n3
    B=SetMaterialAttributes(MaterialAttributes=@n20.MaterialAttributes, BaseColor=@n27, Specular=@n31) {attributeSetTypes [#1, #2]}
    Alpha=@n17
   )) {name=V2.0, nodeColor=<rgba 0 1 0.5313 1>}
  NamedRerouteDeclaration(None=@n3) {name=mat, nodeColor=<rgba 1 0.2109 0 1>}
 ]}
