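# Material function for base customization that adds a cropped emissive mask's glow to an input material's attributes.
# Saturated base emissive, scaled by Darken Base Emissive, is summed with the mask times EmissiveMultiplier.
# The result blends over the original attributes, using the mask's channel maximum as alpha.
# The static switches have no effect, as both branches are zero and the lerp alpha is one; a CheapContrast call is unconnected.
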MaterialFunction /Game/Packages/Fortress_SharedMaterials/Base_Material_Functions/BaseCustomizationFunctions/MF_Base_Customization_CroppedEmissive
graph {nodes [
  n1 = FunctionInput() {inputName=Material, inputType=FunctionInput_MaterialAttributes, sortPriority=-200}
  n2 = GetMaterialAttributes(None=@n1) {attributeGetTypes [#1]}
  n3 = Constant()
  n4 = StaticBoolParameter() {desc=use to bypass the wrap to have emissive enable per-object, group=Emissive, parameterName=Wrap-Override-Emissive Weapons-only}
  n5 = StaticSwitch(
   True=@n3
   False=@n3
   Value=StaticSwitch(True=@n4, False=StaticBool(), Value=StaticBoolParameter() {group=Emissive, parameterName=Get Rid of Emissive Texture})
  )
  n6 = ScalarParameter() {defaultValue=1, desc=Do Not Touch Per-Wepaons Only, group=Emissive, parameterName=Darken Base Emissive}
  MaterialExpressionFunctionInput_0 = FunctionInput() {bUsePreviewValueAsDefault=true, inputName=Cropped Emissive Mask}
  n7 = Add(
   A=Multiply(A=Saturate(None=@n2.EmissiveColor), B=LinearInterpolate(A=@n5, B=@n6) {constAlpha=1})
   B=Multiply(
    A=VectorParameter() {group=Emissive, parameterName=EmissiveMultiplier, defaultValue=<rgba 1 1 1 1>}.RGB
    B=@MaterialExpressionFunctionInput_0
   )
  )
  n8 = Convert(None=@MaterialExpressionFunctionInput_0) {
   nodeName=Break Float3
   convertInputs [{
     type=Vector3
     expressionInput {
      expression=&self:MaterialExpressionFunctionInput_0
      inputName=None
      mask=0
      maskA=0
      maskB=0
      maskG=0
      maskR=0
      outputIndex=0
     }
    }]
   convertOutputs [{type=Scalar}, {type=Scalar}, {type=Scalar}]
  }
  n9 = Max(A=@n8, B=@n8)
  FunctionOutput(None=BlendMaterialAttributes(
    A=@n1
    B=SetMaterialAttributes!(MaterialAttributes=@n2.MaterialAttributes, EmissiveColor=@n7) {attributeSetTypes [#1]}
    Alpha=Max(A=@n9, B=@n8)
   ))
  MaterialFunctionCall(In=@MaterialExpressionFunctionInput_0, Contrast=Constant()) {materialFunction=&CheapContrast}
 ]}
