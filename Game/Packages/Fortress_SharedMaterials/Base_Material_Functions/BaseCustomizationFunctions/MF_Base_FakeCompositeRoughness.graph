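# Material function in the base customization set that fakes composite roughness by adding roughness from a customization crevasse mask.
# It reads Roughness from the incoming material attributes, adds the crevasse input (_CM G) raised to a high constant power, saturates the sum and writes it back.
# A static switch parameter in the Customization group, Disable_Fake_Crevasse_Rouhness, passes the input material through unchanged when true.
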MaterialFunction /Game/Packages/Fortress_SharedMaterials/Base_Material_Functions/BaseCustomizationFunctions/MF_Base_FakeCompositeRoughness
graph {nodes [
  n1 = FunctionInput() {bUsePreviewValueAsDefault=true, inputName=Material, inputType=FunctionInput_MaterialAttributes, previewValue=<1 1 1 1>}
  n2 = GetMaterialAttributes(None=@n1) {attributeGetTypes [#1]}
  n3 = FunctionInput() {bUsePreviewValueAsDefault=true, inputName="Customization Crevasse (_CM G)", inputType=FunctionInput_Scalar, sortPriority=2}
  n4 = Constant() {r=7}
  n5 = Saturate(None=Add(A=@n2.Roughness, B=Power(Base=@n3, Exp=@n4)))
  FunctionOutput(None=StaticSwitchParameter(
    True=@n1
    False=SetMaterialAttributes!(MaterialAttributes=@n2.MaterialAttributes, Roughness=@n5) {attributeSetTypes [#1]}
   ) {group=Customization, parameterName=Disable_Fake_Crevasse_Rouhness, sortPriority=5}) {outputName=Material Result}
 ]}
