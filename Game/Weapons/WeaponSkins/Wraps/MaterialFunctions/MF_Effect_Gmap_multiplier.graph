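# Material function for weapon wraps that multiplies gradient-map-driven tints into base color and emissive, behind a static switch.
# The CM mask's alpha is remapped to drive each tint, inverted for emissive.
# Per-object range overrides, selected by state inputs, feed their G channel into both Low and High.
# Primary and accent masks limit each blend, with optional HotWax masking under Customization V2.
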MaterialFunction /Game/Weapons/WeaponSkins/Wraps/MaterialFunctions/MF_Effect_Gmap_multiplier
graph {nodes [
  n1 = NamedRerouteUsage()
  n2 = TextureSample() {samplerType=SAMPLERTYPE_Normal, texture=/Game/Weapons/FORT_Rifles/T_SCAR_N}
  n3 = Constant3Vector() {constant=<rgba 1 0 1 0>}
  n4 = GetMaterialAttributes(None=FunctionInput(Preview=SetMaterialAttributes!(MaterialAttributes, Normal=@n2.RGB, BaseColor=@n3) {attributeSetTypes [#1, #2]}) {inputName=Material, inputType=FunctionInput_MaterialAttributes}) {attributeGetTypes [
    #2
    #3
    #1
    #4
    #5
   ]}
  n5 = NamedRerouteUsage()
  n6 = ScalarParameter() {defaultValue=0.2, group=Effect_Gmap_Multiplier, parameterName=Effect_Gmap_Low}
  n7 = ScalarParameter() {defaultValue=0.8, group=Effect_Gmap_Multiplier, parameterName=Effect_Gmap_High}
  n8 = MaterialFunctionCall(Color=NamedRerouteUsage()) {materialFunction=&SplitComponents}
  n9 = LinearInterpolate(
   A=VectorParameter() {group=Effect_Gmap_Multiplier, parameterName=Effect_Gmap_ColorTint, defaultValue=<rgba 0 0 0 1>}.RGB
   Alpha=LinearInterpolate(
    A=MaterialFunctionCall(Input=@n5, Input Low=@n6, Input High=@n7) {materialFunction=&RemapValue_0-1}.Result
    B=MaterialFunctionCall(Input=@n5, Input Low=@n8.G, Input High=@n8.G) {materialFunction=&RemapValue_0-1}.Result
    Alpha=@n8.R
   )
  )
  n10 = OneMinus(None=FunctionInput() {inputName=Hotwax_Msk, inputType=FunctionInput_Scalar})
  n11 = TextureSample() {samplerType=SAMPLERTYPE_Masks, texture=/Game/Weapons/FORT_Rifles/Textures/T_Scar_CM}
  n12 = FunctionInput(Preview=@n11.RGBA) {inputName=CM_Msk, inputType=FunctionInput_Vector4, sortPriority=4}
  n13 = MaterialFunctionCall(Float4=@n12) {materialFunction=&BreakOutFloat4Components}
  n14 = Max(
   A=Min(
    A=ScalarParameter() {defaultValue=1, group=Effect_Gmap_Multiplier, parameterName="Effect_Gmap_Primary_Msk  1=Active", sortPriority=0}
    B=NamedRerouteDeclaration(None=@n13.R) {name=Msk_Prime, nodeColor=<rgba 0 1 0.7422 1>}
   )
   B=Min(
    A=NamedRerouteDeclaration(None=@n13.B) {name=Msk_Accent, nodeColor=<rgba 0 1 0.7656 1>}
    B=ScalarParameter() {group=Effect_Gmap_Multiplier, parameterName="Effect_Gmap_Accent_Msk  1=Active", sortPriority=1}
   )
  )
  n15 = Min(
   A=Max(A=@n10, B=ScalarParameter() {defaultValue=1, group=Effect_Gmap_Multiplier, parameterName="Effect_Gmap_HotWax_Msk    0=On", sortPriority=0})
   B=@n14
  )
  n16 = BlendMaterialAttributes(
   A=@n1
   B=SetMaterialAttributes!(MaterialAttributes=NamedRerouteUsage(), BaseColor=Multiply(A=@n4.BaseColor, B=@n9)) {attributeSetTypes [#2]}
   Alpha=StaticSwitch(
    True=StaticSwitchParameter(True=@n15, False=@n14) {group=Customization, parameterName=Effect_HotWax}
    False=@n14
    Value=StaticBoolParameter() {group=Customization, parameterName="Custmization  V2", sortPriority=-500}
   )
  ) {vertexAttributeBlendType=UseA}
  n17 = GetMaterialAttributes(None=@n16) {attributeGetTypes [
    #2
    #3
    #1
    #4
    #5
   ]}
  n18 = VectorParameter() {
   group=Effect_Gmap_Emissive_Multiplier
   parameterName=Effect_Gmap_Emissive_ColorTint
   channelNames {a=Gmap Multplyer, b="", g="", r=""}
   defaultValue=<rgba 0 0 0 1>
  }
  n19 = MaterialFunctionCall(Color=NamedRerouteUsage()) {materialFunction=&SplitComponents}
  n20 = LinearInterpolate(
   A=MaterialFunctionCall(
    Input=@n5
    Input Low=ScalarParameter() {defaultValue=0.2, group=Effect_Gmap_Emissive_Multiplier, parameterName=Effect_Gmap_Emissive_Low}
    Input High=ScalarParameter() {defaultValue=1, group=Effect_Gmap_Emissive_Multiplier, parameterName=Effect_Gmap_Emissive_High}
   ) {materialFunction=&RemapValue_0-1}.Result
   B=MaterialFunctionCall(Input=@n5, Input Low=@n19.G, Input High=@n19.G) {materialFunction=&RemapValue_0-1}.Result
   Alpha=@n19.R
  )
  n21 = Saturate(None=Multiply(A=OneMinus(None=@n20), B=@n18.Gmap Multplyer))
  n22 = SetMaterialAttributes!(MaterialAttributes=@n16, EmissiveColor=Multiply(A=@n17.EmissiveColor, B=LinearInterpolate(B=@n18, Alpha=@n21) {constA=1})) {attributeSetTypes [#5]}
  n23 = ScalarParameter() {defaultValue=1, group=Effect_Gmap_Emissive_Multiplier, parameterName="Effect_Gmap_Emissive_HotWax_Msk    0=On", sortPriority=0}
  n24 = Max(A=@n10, B=@n23)
  n25 = MaterialFunctionCall(Float4=@n12) {materialFunction=&BreakOutFloat4Components}
  n26 = Max(
   A=Min(
    A=ScalarParameter() {group=Effect_Gmap_Emissive_Multiplier, parameterName="Effect_Gmap_Emissive_Primary_Msk  1=Active", sortPriority=0}
    B=NamedRerouteDeclaration(None=@n25.R) {name=Msk_Prime_1, nodeColor=<rgba 0 1 0.7422 1>}
   )
   B=Min(
    A=NamedRerouteDeclaration(None=@n25.B) {name=Msk_Accent_1, nodeColor=<rgba 0 1 0.7656 1>}
    B=ScalarParameter() {group=Effect_Gmap_Emissive_Multiplier, parameterName="Effect_Gmap_Emissive_Accent_Msk  1=Active", sortPriority=1}
   )
  )
  n27 = StaticSwitch(
   True=StaticSwitchParameter(True=Min(A=@n24, B=@n26), False=@n26) {group=Customization, parameterName=Effect_HotWax}
   False=@n26
   Value=StaticBoolParameter() {group=Customization, parameterName="Custmization  V2", sortPriority=-500}
  )
  FunctionOutput(None=StaticSwitchParameter(True=BlendMaterialAttributes(A=@n16, B=@n22, Alpha=@n27) {vertexAttributeBlendType=UseA}, False=@n1) {group=Customization, parameterName=Effect_Gmap_Multiplier})
  LinearInterpolate(A=@n6, B=@n7, Alpha=@n5)
  NamedRerouteDeclaration(None=@n13.A) {name=Msk_Gmap, nodeColor=<rgba 0 1 0.7656 1>}
  NamedRerouteDeclaration(None=@n4.MaterialAttributes) {name=Root_Material, nodeColor=<rgba 0.7812 1 0 1>}
  NamedRerouteDeclaration(None=@n11.RGBA) {name=CM_Mask, nodeColor=<rgba 1 0 0.3047 1>}
  NamedRerouteDeclaration(None=@n25.A) {name=Msk_Gmap_1, nodeColor=<rgba 0 1 0.7656 1>}
  NamedRerouteDeclaration(None=FunctionInput() {inputName=is state, inputType=FunctionInput_Vector4}) {name=IsState, nodeColor=<rgba 0.1497 1 0.1633 1>}
  NamedRerouteDeclaration(None=FunctionInput() {inputName=is state_B, inputType=FunctionInput_Vector2}) {name=IsState_B, nodeColor=<rgba 0.1497 1 0.1633 1>}
  n28 = VectorParameter() {
   group=Effect_Gmap_Multiplier | Objects
   parameterName=Effect_Gmap_Range_Weapon
   channelNames {a="", b=High, g=Low, r=Use}
   defaultValue=<rgba 0 0.2 0.8 1>
  }
  n29 = VectorParameter() {
   group=Effect_Gmap_Multiplier | Objects
   parameterName=Effect_Gmap_Range_Vehicle
   channelNames {a="", b=high, g=low, r=use}
   defaultValue=<rgba 0 0.2 0.8 1>
  }
  n30 = VectorParameter() {
   group=Effect_Gmap_Multiplier | Objects
   parameterName=Effect_Gmap_Range_VehicleSide
   channelNames {a="", b=High, g=Low, r=Use}
   defaultValue=<rgba 0 0.2 0.8 1>
  }
  n31 = MaterialFunctionCall(Float4=NamedRerouteUsage()) {materialFunction=&BreakOutFloat4Components}
  n32 = VectorParameter() {
   group=Effect_Gmap_Multiplier | Objects
   parameterName=Effect_Gmap_Range_Mech
   channelNames {a="", b=High, g=Low, r=Use}
   defaultValue=<rgba 0 0.2 0.8 1>
  }
  n33 = VectorParameter() {
   group=Effect_Gmap_Multiplier | Objects
   parameterName=Effect_Gmap_Range_Character
   channelNames {a="", b=High, g=Low, r=Use}
   defaultValue=<rgba 0 0.2 0.8 1>
  }
  n34 = LinearInterpolate(
   A=LinearInterpolate(A=@n28.RGB, B=LinearInterpolate(A=@n29.RGB, B=@n30.RGB, Alpha=@n31.G), Alpha=@n31.R)
   B=LinearInterpolate(A=@n32.RGB, B=@n33.RGB, Alpha=@n31.A)
   Alpha=@n31.B
  )
  n35 = VectorParameter() {
   group=Effect_Gmap_Multiplier | Objects
   parameterName=Effect_Gmap_Range_Backpack
   channelNames {a="", b=High, g=Low, r=Use}
   defaultValue=<rgba 0 0.2 0.8 1>
  }
  n36 = MaterialFunctionCall(Float2=NamedRerouteUsage()) {materialFunction=&BreakOutFloat2Components}
  n37 = VectorParameter() {
   group=Effect_Gmap_Multiplier | Objects
   parameterName=Effect_Gmap_Range_Pickaxe
   channelNames {a="", b=High, g=Low, r=Use}
   defaultValue=<rgba 0 0.2 0.8 1>
  }
  NamedRerouteDeclaration(None=LinearInterpolate(A=LinearInterpolate(A=@n34, B=@n35.RGB, Alpha=@n36.R), B=@n37.RGB, Alpha=@n36.G)) {name=Range, nodeColor=<rgba 0 0.625 1 1>}
  n38 = VectorParameter() {
   group=Effect_Gmap_Emissive_Multiplier | Objects
   parameterName=Effect_Gmap_Emissive_Range_Weapon
   channelNames {a="", b=High, g=Low, r=Use}
   defaultValue=<rgba 0 0.2 0.8 1>
  }
  n39 = VectorParameter() {
   group=Effect_Gmap_Emissive_Multiplier | Objects
   parameterName=Effect_Gmap_Emissive_Range_Vehicle
   channelNames {a="", b=high, g=low, r=use}
   defaultValue=<rgba 0 0.2 0.8 1>
  }
  n40 = VectorParameter() {
   group=Effect_Gmap_Emissive_Multiplier | Objects
   parameterName=Effect_Gmap_Emissive_Range_VehicleSide
   channelNames {a="", b=High, g=Low, r=Use}
   defaultValue=<rgba 0 0.2 0.8 1>
  }
  n41 = MaterialFunctionCall(Float4=NamedRerouteUsage()) {materialFunction=&BreakOutFloat4Components}
  n42 = VectorParameter() {
   group=Effect_Gmap_Emissive_Multiplier | Objects
   parameterName=Effect_Gmap_Emissive_Range_Mech
   channelNames {a="", b=High, g=Low, r=Use}
   defaultValue=<rgba 0 0.2 0.8 1>
  }
  n43 = VectorParameter() {
   group=Effect_Gmap_Emissive_Multiplier | Objects
   parameterName=Effect_Gmap_Emissive_Range_Character
   channelNames {a="", b=High, g=Low, r=Use}
   defaultValue=<rgba 0 0.2 0.8 1>
  }
  n44 = LinearInterpolate(
   A=LinearInterpolate(A=@n38.RGB, B=LinearInterpolate(A=@n39.RGB, B=@n40.RGB, Alpha=@n41.G), Alpha=@n41.R)
   B=LinearInterpolate(A=@n42.RGB, B=@n43.RGB, Alpha=@n41.A)
   Alpha=@n41.B
  )
  n45 = VectorParameter() {
   group=Effect_Gmap_Emissive_Multiplier | Objects
   parameterName=Effect_Gmap_Emissive_Range_Backpack
   channelNames {a="", b=High, g=Low, r=Use}
   defaultValue=<rgba 0 0.2 0.8 1>
  }
  n46 = MaterialFunctionCall(Float2=NamedRerouteUsage()) {materialFunction=&BreakOutFloat2Components}
  n47 = VectorParameter() {
   group=Effect_Gmap_Emissive_Multiplier | Objects
   parameterName=Effect_Gmap_Emissive_Range_Pickaxe
   channelNames {a="", b=High, g=Low, r=Use}
   defaultValue=<rgba 0 0.2 0.8 1>
  }
  NamedRerouteDeclaration(None=LinearInterpolate(A=LinearInterpolate(A=@n44, B=@n45.RGB, Alpha=@n46.R), B=@n47.RGB, Alpha=@n46.G)) {name=Emissive Range, nodeColor=<rgba 0 0.625 1 1>}
 ]}
