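# MF_Per_ObjectMaskOverride is a material function for weapon skin wraps that blends a root material attribute set toward a second Material set.
# The blend alpha is the saturated sum of three scalar inputs: Per_Mask_Value, Mask Primary and Mask Accent.
# Per_Mask_Value defaults to one, so when unconnected it saturates the alpha and Material fully replaces Mat_Root regardless of the other masks.
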MaterialFunction /Game/Weapons/WeaponSkins/Wraps/MaterialFunctions/MF_Per_ObjectMaskOverride
graph {nodes [
  n1 = FunctionInput(Preview=Constant() {r=1}) {bUsePreviewValueAsDefault=true, inputName=Per_Mask_Value, inputType=FunctionInput_Scalar, sortPriority=10, previewValue=<1 0 0 1>}
  n2 = Add(
   A=@n1
   B=Add(
    A=FunctionInput() {inputName=Mask Primary, inputType=FunctionInput_Scalar, sortPriority=10}
    B=FunctionInput() {inputName=Mask Accent, inputType=FunctionInput_Scalar, sortPriority=10}
   )
  )
  FunctionOutput(None=BlendMaterialAttributes(
    A=FunctionInput() {inputName=Mat_Root, inputType=FunctionInput_MaterialAttributes}
    B=FunctionInput() {inputName=Material, inputType=FunctionInput_MaterialAttributes}
    Alpha=Saturate(None=@n2)
   ))
 ]}
