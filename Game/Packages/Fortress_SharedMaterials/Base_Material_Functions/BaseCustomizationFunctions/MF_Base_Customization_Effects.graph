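# Material function that layers wrap and base customization effects onto input material attributes.
# It chains Gmap multiplier, TOD, fake composite roughness, cropped emissive and crevice overlay functions, ending in a per-object mask override against a root material.
# CM mask channels are split into primary, crevice, secondary and Gmap masks, shared through named reroutes.
# The default calculated-bounds input subtracts bounds Max from itself, giving a zero divisor.
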MaterialFunction /Game/Packages/Fortress_SharedMaterials/Base_Material_Functions/BaseCustomizationFunctions/MF_Base_Customization_Effects
graph {nodes [
  n1 = NamedRerouteUsage()
  n2 = NamedRerouteUsage()
  n3 = NamedRerouteUsage()
  n4 = NamedRerouteUsage()
  n5 = NamedRerouteUsage()
  n6 = MaterialFunctionCall!(
   Material=MaterialFunctionCall!(Material=MaterialFunctionCall!(is state_B=@n1, Hotwax_Msk=@n2, is state=@n3, Material=@n4, CM_Msk=@n5) {materialFunction=/Game/Weapons/WeaponSkins/Wraps/MaterialFunctions/MF_Effect_Gmap_multiplier}.Result) {materialFunction=/Game/Weapons/WeaponSkins/Wraps/MaterialFunctions/MF_Effect_TOD}.Result
   "Customization Crevasse (_CM G)"=NamedRerouteUsage()
  ) {materialFunction=&/MF_Base_FakeCompositeRoughness}
  n7 = MaterialFunctionCall!(Material=@n6.Material Result, Cropped Emissive Mask=NamedRerouteUsage()) {materialFunction=&/MF_Base_Customization_CroppedEmissive}
  n8 = MaterialFunctionCall!(
   In=@n7.Result
   Customization CM=NamedRerouteUsage()
   Customization Crevice=NamedRerouteUsage()
   "Customization Mask 1 (_CM B Top Left)"=NamedRerouteUsage()
   "Customization Mask 2 (_CM B Top Right)"=NamedRerouteUsage()
  ) {materialFunction=&/MF_Base_Customization_CreviceOverlay}
  FunctionOutput(None=MaterialFunctionCall!(
    Mat_Root=NamedRerouteUsage()
    Material=@n8.Result
    Mask Accent=NamedRerouteUsage()
    Mask Primary=NamedRerouteUsage()
    Per_Mask_Value=NamedRerouteUsage()
   ) {materialFunction=/Game/Weapons/WeaponSkins/Wraps/MaterialFunctions/MF_Per_ObjectMaskOverride}.Result)
  NamedRerouteDeclaration(None=FunctionInput() {inputName=Material, inputType=FunctionInput_MaterialAttributes, sortPriority=-100}) {name=In_Material, nodeColor=<rgba 1 0 0.1641 1>}
  n9 = MaterialFunctionCall(Float4=NamedRerouteDeclaration(None=FunctionInput() {inputName=CM_Msk, inputType=FunctionInput_Vector4, sortPriority=4}) {name=Masks_RGBA_CM, nodeColor=<rgba 1 0.1875 0 1>}) {materialFunction=&BreakOutFloat4Components}
  NamedRerouteDeclaration(None=@n9.B) {name="Mask_Secondary {B Top Right}", nodeColor=<rgba 0.0094 0.0094 0.901 1>}
  NamedRerouteDeclaration(None=FunctionInput(Preview=Constant3Vector() {constant=<rgba 1 0 0 0>}) {bUsePreviewValueAsDefault=true, inputName=ForwardAxis}) {name=Forward_Axies, nodeColor=<rgba 0 0.9297 1 1>}
  n10 = Constant4Vector() {constant=<rgba 0 0 0 0.5>}
  NamedRerouteDeclaration(None=FunctionInput(Preview=@n10.RGBA) {bUsePreviewValueAsDefault=true, inputName=Object-SafeZone-Back, inputType=FunctionInput_Vector4, sortPriority=100}) {name=SafeZone_Back, nodeColor=<rgba 0 0.9297 1 1>}
  NamedRerouteDeclaration(None=@n9.R) {name="Mask_Primary {B Top Left}", nodeColor=<rgba 0.0094 0.0094 0.901 1>}
  NamedRerouteDeclaration(None=@n9.G) {name=Crevice_Mask, nodeColor=<rgba 0 0.7865 0 1>}
  NamedRerouteDeclaration(None=FunctionInput() {inputName=Root_Material, inputType=FunctionInput_MaterialAttributes, sortPriority=-200}) {name=Root_Mat, nodeColor=<rgba 0.8047 1 0 1>}
  NamedRerouteDeclaration(None=FunctionInput(Preview=@n10.RGBA) {bUsePreviewValueAsDefault=true, inputName=Object-SafeZone-Middle, inputType=FunctionInput_Vector4, sortPriority=101}) {name=SafeZone_Middle, nodeColor=<rgba 0 0.9297 1 1>}
  NamedRerouteDeclaration(None=FunctionInput(Preview=@n10.RGBA) {bUsePreviewValueAsDefault=true, inputName=Object-SafeZone-Front, inputType=FunctionInput_Vector4, sortPriority=102}) {name=SafeZone_Front, nodeColor=<rgba 0 0.9297 1 1>}
  NamedRerouteDeclaration(None=FunctionInput() {bUsePreviewValueAsDefault=true, inputName=Per_ObjectMaskOverride, inputType=FunctionInput_Scalar, sortPriority=75}) {name=PerObjectMask, nodeColor=<rgba 1 0.6569 0.0791 1>}
  NamedRerouteDeclaration(None=FunctionInput() {bUsePreviewValueAsDefault=true, inputName=IsState_BV4, inputType=FunctionInput_Vector4, sortPriority=4}) {name=Is_StatsB_V4, nodeColor=<rgba 0 0.9253 1 1>}
  NamedRerouteDeclaration(None=FunctionInput() {inputName=Hotwax_Msk, inputType=FunctionInput_Scalar}) {name=HotWax_Msk, nodeColor=<rgba 0 0.2266 1 1>}
  NamedRerouteDeclaration(None=FunctionInput() {bUsePreviewValueAsDefault=true, inputName=Cropped Emissive Mask, sortPriority=20}) {name=Cropped Emissive Mask, nodeColor=<rgba 1 0.0938 0 1>}
  NamedRerouteDeclaration(None=FunctionInput(Preview=Constant4Vector().RGBA) {bUsePreviewValueAsDefault=true, inputName=IsState, inputType=FunctionInput_Vector4, sortPriority=4}) {name=IsState, nodeColor=<rgba 0.1497 1 0.1633 1>}
  NamedRerouteDeclaration(None=@n9.A) {name=Mask_Gmap, nodeColor=<rgba 0.0094 0.0094 0.901 1>}
  NamedRerouteDeclaration(None=FunctionInput(Preview=Constant2Vector()) {bUsePreviewValueAsDefault=true, inputName=IsState_B, inputType=FunctionInput_Vector2, sortPriority=4}) {name=IsState_B, nodeColor=<rgba 0.1497 1 0.1633 1>}
  n11 = PreSkinnedLocalBounds()
  n12 = Subtract(A=VertexInterpolator(VS=PreSkinnedPosition()).PS, B=@n11.Min)
  n13 = MaterialFunctionCall(Color=Subtract(A=@n11.Max, B=@n11.Max)) {materialFunction=&SplitComponents}
  n14 = Max(A=Max(A=@n13.R, B=@n13.R), B=@n13.B)
  NamedRerouteDeclaration(None=FunctionInput(Preview=Divide(A=@n12, B=@n14)) {bUsePreviewValueAsDefault=true, inputName=Calcated_Bounds, sortPriority=11}) {name=Caclcated_Bounds, nodeColor=<rgba 0 0.9297 1 1>}
  n15 = PreSkinnedNormal()
  NamedRerouteDeclaration(None=FunctionInput(Preview=VertexInterpolator(VS=@n15).PS) {bUsePreviewValueAsDefault=true, inputName=Pre-Skinned Local Normal, sortPriority=10}) {name=PreSkin_Normal, nodeColor=<rgba 0.8281 1 0 1>}
 ]}
