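# MF_JacketWrap is a customization material function that optionally applies the MF_Jacket function from the CycloneUniverse character body to incoming material attributes.
# The useJacket static switch, in the Customization group, outputs either the MF_Jacket result or the unchanged input attributes.
# MF_Jacket receives the Alpha Mask and GMAP scalar inputs plus the material attributes.
# A SetMaterialAttributes node overriding BaseColor with a constant does not feed the output.
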MaterialFunction /Game/Packages/Fortress_SharedMaterials/Base_Material_Functions/BaseCustomizationFunctions/MF_JacketWrap
graph {nodes [
  n1 = FunctionInput() {inputName=Alpha Mask, inputType=FunctionInput_Scalar}
  n2 = FunctionInput() {inputName=GMAP, inputType=FunctionInput_Scalar}
  n3 = FunctionInput() {inputType=FunctionInput_MaterialAttributes}
  n4 = GetMaterialAttributes(None=@n3)
  FunctionOutput(None=StaticSwitchParameter(
    True=MaterialFunctionCall!(Alpha=@n1, "Gmap "=@n2, Material=@n4.MaterialAttributes) {materialFunction=/Game/Characters/Player/Male/Medium/Bodies/M_MED_CycloneUniverse/Materials/MF_Jacket}.Result
    False=@n3
   ) {group=Customization, parameterName=useJacket})
  SetMaterialAttributes!(MaterialAttributes=@n4.MaterialAttributes, BaseColor=Constant()) {attributeSetTypes [#1]}
 ]}
