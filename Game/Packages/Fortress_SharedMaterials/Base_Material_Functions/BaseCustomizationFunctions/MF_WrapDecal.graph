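# Material function that projects a decal texture onto the base color of wrapped weapons and vehicles, behind the UseWrapDecalOnFront static switch.
# Radius-normalized pre-skinned position gets offset and scale, with isVehicle blending weapon and vehicle settings and choosing a clamped YZ or XZ planar projection.
# Decal alpha times the customization mask drives the blend.
# Vehicles also fade it on vertically facing surfaces using a contrast-adjusted normal Z mask.
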MaterialFunction /Game/Packages/Fortress_SharedMaterials/Base_Material_Functions/BaseCustomizationFunctions/MF_WrapDecal
graph {nodes [
  n1 = FunctionInput() {inputName=Material, inputType=FunctionInput_MaterialAttributes, sortPriority=-200}
  n2 = GetMaterialAttributes(None=@n1) {attributeGetTypes [#1]}
  n3 = FunctionInput(Preview=VertexInterpolator(VS=PreSkinnedPosition()).PS) {bUsePreviewValueAsDefault=true, inputName=Pre-Skinned Local Position, sortPriority=41}
  n4 = ObjectRadius()
  n5 = VectorParameter() {group=Primary Triplaner Decal, parameterName=DecalOffset_Weapon, sortPriority=1, defaultValue=<rgba 0 -0.4 -0.65 1>}
  n6 = VectorParameter() {group=Primary Triplaner Decal, parameterName=DecalOffset_Vehicle, sortPriority=5, defaultValue=<rgba 0.2 0 -0.65 1>}
  n7 = FunctionInput() {inputName=isVehicle, inputType=FunctionInput_Scalar}
  n8 = VectorParameter() {group=Primary Triplaner Decal, parameterName=DecalScale_Weapon, sortPriority=2, defaultValue=<rgba 1.5 1 -1.5 1>}
  n9 = VectorParameter() {group=Primary Triplaner Decal, parameterName=DecalScale_Vehicle, sortPriority=4, defaultValue=<rgba 1 1 -1 1>}
  n10 = Multiply(A=Add(A=Divide(A=@n3, B=@n4), B=LinearInterpolate(A=@n5.RGB, B=@n6.RGB, Alpha=@n7)), B=LinearInterpolate(A=@n8.RGB, B=@n9.RGB, Alpha=@n7))
  n11 = ComponentMask(None=@n10) {b=true, g=true}
  n12 = ComponentMask(None=@n10) {b=true, r=true}
  n13 = TextureSampleParameter2D(UVs=Clamp(None=LinearInterpolate(A=@n11, B=@n12, Alpha=@n7))) {
   group=Primary Triplaner Decal
   parameterName=Decal
   samplerSource=SSM_Clamp_WorldGroupSettings
   texture=/Game/Weapons/WeaponSkins/Wraps/Textures/CamoBig/T_PlaneBigCamo_Mouth
  }
  n14 = FunctionInput() {bUsePreviewValueAsDefault=true, inputName="Customization Mask (_CM B)", inputType=FunctionInput_Scalar, sortPriority=24}
  n15 = Abs(None=MaterialFunctionCall(Float3=FunctionInput() {bUsePreviewValueAsDefault=true, inputName=Pre-Skinned Local Normal, sortPriority=42}) {materialFunction=&BreakOutFloat3Components}.B)
  n16 = ScalarParameter() {defaultValue=0.7, group=Primary Triplaner Decal, parameterName=DecalVerticalMaskGain}
  n17 = OneMinus(None=MaterialFunctionCall(
    In=Multiply(A=@n15, B=@n16)
    Contrast=ScalarParameter() {defaultValue=1.2, group=Primary Triplaner Decal, parameterName=DecalVerticalMaskContrast}
   ) {materialFunction=&CheapContrast}.Result)
  n18 = LinearInterpolate(A=@n2.BaseColor, B=@n13.RGB, Alpha=Multiply(A=Multiply(A=@n13.A, B=@n14), B=LinearInterpolate(B=@n17, Alpha=@n7) {constA=1}))
  FunctionOutput(None=StaticSwitchParameter(
    True=SetMaterialAttributes!(MaterialAttributes=@n2.MaterialAttributes, BaseColor=@n18) {attributeSetTypes [#1]}
    False=@n1
   ) {group=Customization, parameterName=UseWrapDecalOnFront, sortPriority=-10})
  Multiply()
 ]}
