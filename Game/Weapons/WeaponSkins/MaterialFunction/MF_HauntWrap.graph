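# Material function for the Haunt weapon wrap, adding animated ghostly glow, smoke, stars and lightning to weapons, vehicles and mechs.
# Projects camera-facing virtual-plane UVs around object bounds, with per-type parameters chosen by isVehicle and isMech.
# Sphere masks form glows and swirl UVs that drive flow-distorted panning layers and offset-sampled turbulence.
# UseHauntWrap blends the result into incoming attributes by customization mask; pivot, volume-shadow and nebula-texture branches are unconnected.
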MaterialFunction /Game/Weapons/WeaponSkins/MaterialFunction/MF_HauntWrap
graph {nodes [
  n1 = VectorParameter() {group=HauntWrap, parameterName=WeaponPivotAdjustments, defaultValue=<rgba 0 0 0 1>}
  n2 = VectorParameter() {group=HauntWrap, parameterName=VehiclePivotAdjustments, defaultValue=<rgba 0 0 0 1>}
  n3 = FunctionInput(Preview=Constant()) {bUsePreviewValueAsDefault=true, inputName=isVehicle, inputType=FunctionInput_Scalar}
  n4 = FunctionInput(Preview=Constant()) {bUsePreviewValueAsDefault=true, inputName=isMech, inputType=FunctionInput_Scalar}
  Add(
   A=MaterialFunctionCall() {materialFunction=&ObjectPivotPoint}.Object Pivot Location
   B=LinearInterpolate(
    A=LinearInterpolate(A=@n1.RGB, B=@n2.RGB, Alpha=@n3)
    B=VectorParameter() {group=HauntWrap, parameterName=MechPivotAdjustments, defaultValue=<rgba 0 0 0 1>}.RGB
    Alpha=@n4
   )
  )
  n5 = PreSkinnedLocalBounds()
  n6 = Add(A=@n5.Min, B=FunctionInput() {inputName=Object Bounds Min Adjustment, sortPriority=51}) {constB=0}
  n7 = FunctionInput() {inputName=Object Bounds Max Adjustment, sortPriority=52}
  n8 = Multiply(A=Subtract(A=Add(A=@n5.Max, B=@n7) {constB=0}, B=@n6)) {constB=0.5}
  n9 = LinearInterpolate(A=TransformPosition(None=Add(A=@n6, B=@n8)) {transformType=TRANSFORMPOSSOURCE_World}, B=ObjectPositionWS(), Alpha=@n4)
  n10 = Normalize(VectorInput=Subtract(A=@n9, B=CameraPositionWS()))
  n11 = VectorParameter() {group=HauntWrap, parameterName=Axis, defaultValue=<rgba 0 0 1 1>}
  n12 = LinearInterpolate(
   A=LinearInterpolate(
    A=ScalarParameter() {defaultValue=100, group=HauntWrap, parameterName=WeaponUVRadialScaleZero}
    B=ScalarParameter() {defaultValue=100, group=HauntWrap, parameterName=VehicleUVRadialScaleZero}
    Alpha=@n3
   )
   B=ScalarParameter() {defaultValue=100, group=HauntWrap, parameterName=MechUVRadialScaleZero}
   Alpha=@n4
  )
  n13 = ScalarParameter() {defaultValue=100, group=HauntWrap, parameterName=BackpackUVRadialScaleZero}
  n14 = NamedRerouteUsage()
  n15 = MaterialFunctionCall!(
   Plane Normal Axis=@n10
   Plane X Axis=Normalize(VectorInput=CrossProduct(A=@n10, B=@n11.RGB))
   Plane Center=@n9
   UV Scale=Divide(A=ObjectRadius(), B=LinearInterpolate(A=@n12, B=@n13, Alpha=@n14))
  ) {materialFunction=/Game/Athena/Environments/Nexus/Materials/VirtualPlaneCoordinatesAbs}
  n16 = MaterialFunctionCall(Float2=@n15.UVs) {materialFunction=&BreakOutFloat2Components}
  n17 = OneMinus(None=@n16.R)
  n18 = ObjectOrientation()
  n19 = Clamp(None=DotProduct(
    A=MaterialFunctionCall!(Vector=@n18) {materialFunction=/Engine/Functions/Engine_MaterialFunctions01/Vectors/MF_RotateVector_90}.CCW Y
    B=MaterialFunctionCall() {materialFunction=&CameraDirectionVector}.Result
   ))
  n20 = Multiply(A=LinearInterpolate(A=@n15.UVs, B=AppendVector(A=@n17, B=@n16.G), Alpha=Ceil(None=@n19))) {constB=2}
  n21 = LinearInterpolate(
   A=VectorParameter() {group=HauntWrap, parameterName=WeaponCorePosition, defaultValue=<rgba 0.5 0.5 0 1>}.RGB
   B=VectorParameter() {group=HauntWrap, parameterName=VehicleCorePosition, defaultValue=<rgba 0.5 0.5 0 1>}.RGB
   Alpha=@n3
  )
  n22 = VectorParameter() {group=HauntWrap, parameterName=MechCorePosition, defaultValue=<rgba 0.5 0.5 0 1>}
  n23 = Subtract(
   A=LinearInterpolate(A=@n15.UVs, B=Subtract(A=@n20), Alpha=ScalarParameter() {group=HauntWrap, parameterName=EnableFlipG})
   B=ComponentMask(None=LinearInterpolate(A=@n21, B=@n22.RGB, Alpha=@n4)) {g=true, r=true}
  )
  n24 = ComponentMask(None=Constant3Vector() {constant=<rgba 0.5 0.5 0 0>}) {g=true, r=true}
  n25 = LinearInterpolate(
   A=LinearInterpolate(
    A=ScalarParameter() {defaultValue=1, group=HauntWrap, parameterName=WeaponInnerGlowRadius}
    B=ScalarParameter() {defaultValue=1, group=HauntWrap, parameterName=VehicleInnerGlowRadius}
    Alpha=@n3
   )
   B=ScalarParameter() {defaultValue=1, group=HauntWrap, parameterName=MechInnerGlowRadius}
   Alpha=@n4
  )
  n26 = Constant()
  n27 = Power(Base=Clamp(None=SphereMask(A=@n23, B=@n24, Radius=@n25, Hardness=@n26)))
  n28 = ScalarParameter() {defaultValue=0.75, group=HauntWrap, parameterName=WeaponSwirlRadius}
  n29 = ScalarParameter() {defaultValue=0.75, group=HauntWrap, parameterName=VehicleSwirlRadius}
  n30 = SphereMask(
   A=@n23
   B=@n24
   Radius=LinearInterpolate(
    A=LinearInterpolate(A=@n28, B=@n29, Alpha=@n3)
    B=ScalarParameter() {defaultValue=0.75, group=HauntWrap, parameterName=MechSwirlRadius}
    Alpha=@n4
   )
   Hardness=Constant()
  )
  n31 = Rotator(Coordinate=@n23, Time=Multiply(A=@n30, B=ScalarParameter() {defaultValue=6, group=HauntWrap, parameterName=CoreSwirlStrength}))
  n32 = LinearInterpolate(
   A=LinearInterpolate(
    A=ScalarParameter() {defaultValue=0.5, group=HauntWrap, parameterName=WeaponSmokeTiling}
    B=ScalarParameter() {defaultValue=0.5, group=HauntWrap, parameterName=VehicleSmokeTiling}
    Alpha=@n3
   )
   B=ScalarParameter() {defaultValue=0.5, group=HauntWrap, parameterName=MechSmokeTiling}
   Alpha=@n4
  )
  n33 = LinearInterpolate(
   A=LinearInterpolate(
    A=VectorParameter() {group=HauntWrap, parameterName=WeaponSmokePanSpeedandDirection, defaultValue=<rgba 0 -0.8 0 1>}.RGB
    B=VectorParameter() {group=HauntWrap, parameterName=VehicleSmokePanSpeedandDirection, defaultValue=<rgba 0 -0.8 0 1>}.RGB
    Alpha=@n3
   )
   B=VectorParameter() {group=HauntWrap, parameterName=MechSmokePanSpeedandDirection, defaultValue=<rgba 0 -0.8 0 1>}.RGB
   Alpha=@n4
  )
  n34 = LinearInterpolate(
   A=LinearInterpolate(
    A=ScalarParameter() {defaultValue=0.5, group=HauntWrap, parameterName=WeaponSmokeSwirlTiling}
    B=ScalarParameter() {defaultValue=0.5, group=HauntWrap, parameterName=VehicleSmokeSwirlTiling}
    Alpha=@n3
   )
   B=ScalarParameter() {defaultValue=0.5, group=HauntWrap, parameterName=MechSmokeSwirlTiling}
   Alpha=@n4
  )
  n35 = LinearInterpolate(
   A=LinearInterpolate(
    A=VectorParameter() {group=HauntWrap, parameterName=WeaponSwirlPanSpeedandDirection, defaultValue=<rgba 0 -0.5 0 1>}.RGB
    B=VectorParameter() {group=HauntWrap, parameterName=VehicleSwirlPanSpeedandDirection, defaultValue=<rgba 0 -0.5 0 1>}.RGB
    Alpha=@n3
   )
   B=VectorParameter() {group=HauntWrap, parameterName=MechSwirlPanSpeedandDirection, defaultValue=<rgba 0 -0.5 0 1>}.RGB
   Alpha=@n4
  )
  n36 = TextureSample(UVs=Panner(Coordinate=Multiply(A=@n31, B=@n34), Speed=ComponentMask(None=@n35) {g=true, r=true}) {speedY=-0.5}) {
   samplerSource=SSM_Wrap_WorldGroupSettings
   samplerType=SAMPLERTYPE_LinearColor
   texture=/Game/Global/Textures/Default/Flowmaps/T_SwirlFlow_02
  }
  n37 = Multiply(A=@n36.RGB) {constB=2}
  n38 = ComponentMask(None=Multiply(A=Subtract(A=@n37), B=ScalarParameter() {defaultValue=0.1, group=HauntWrap, parameterName=DistortionStrength})) {g=true, r=true}
  n39 = Add(A=Panner(Coordinate=Multiply(A=@n31, B=@n32), Speed=ComponentMask(None=@n33) {g=true, r=true}) {speedY=-0.8}, B=@n38)
  n40 = TextureSampleParameter2D(UVs=@n39) {
   parameterName=T_FX_Pattern
   samplerSource=SSM_Wrap_WorldGroupSettings
   texture=/Game/Global/Textures/Default/Tileable_Patterns/T_Thick_Smoke_05
  }
  n41 = ComponentMask(None=@n40.RGB) {r=true}
  n42 = ScalarParameter() {defaultValue=0.5, group=HauntWrap, parameterName=GlowSmokiness_NoiseStrength}
  n43 = Clamp(None=Subtract(A=@n27, B=Multiply(A=@n41, B=@n42)))
  n44 = VectorParameter() {group=HauntWrap, parameterName=InnerGlowColor, defaultValue=<rgba 0 0.5703 1 1>}
  n45 = ScalarParameter() {defaultValue=2, group=HauntWrap, parameterName=WeaponGlowFalloffRadius}
  n46 = ScalarParameter() {defaultValue=2, group=HauntWrap, parameterName=VehicleGlowFalloffRadius}
  n47 = SphereMask(
   A=@n23
   B=@n24
   Radius=LinearInterpolate(
    A=LinearInterpolate(A=@n45, B=@n46, Alpha=@n3)
    B=ScalarParameter() {defaultValue=2, group=HauntWrap, parameterName=MechGlowFalloffRadius}
    Alpha=@n4
   )
   Hardness=Constant()
  )
  n48 = Subtract(
   A=Power(Base=Clamp(None=@n47))
   B=Multiply(A=@n41, B=ScalarParameter() {defaultValue=0.5, group=HauntWrap, parameterName=GlowwFalloff_NoiseStrength})
  )
  n49 = Add(
   A=Multiply(A=Multiply(A=@n43, B=@n44.RGB), B=ScalarParameter() {defaultValue=2, group=HauntWrap, parameterName=InnerGlowStrength})
   B=LinearInterpolate(
    A=VectorParameter() {group=HauntWrap, parameterName=GeneralGlowColor, defaultValue=<rgba 0 0.0356 0.0625 1>}.RGB
    B=VectorParameter() {group=HauntWrap, parameterName=FalloffGlowColor, defaultValue=<rgba 0 0.1574 0.276 1>}.RGB
    Alpha=Clamp(None=@n48)
   )
  )
  n50 = ScalarParameter() {defaultValue=0.5, group=HauntWrap, parameterName=WeaponStars_Tiling}
  n51 = ScalarParameter() {defaultValue=0.5, group=HauntWrap, parameterName=VehicleStars_Tiling}
  n52 = Multiply(
   A=Subtract(A=@n24, B=@n23)
   B=LinearInterpolate(
    A=LinearInterpolate(A=@n50, B=@n51, Alpha=@n3)
    B=ScalarParameter() {defaultValue=0.5, group=HauntWrap, parameterName=MechStars_Tiling}
    Alpha=@n4
   )
  )
  n53 = Rotator(Coordinate=@n52, Time=Multiply(A=@n30, B=ScalarParameter() {defaultValue=3, group=ViperZero, parameterName=StarSwirlStrength}))
  n54 = VectorParameter() {group=HauntWrap, parameterName=WeaponStarsPanSpeedandDirection, defaultValue=<rgba 0 0.5 0 1>}
  n55 = VectorParameter() {group=HauntWrap, parameterName=VehicleStarsPanSpeedandDirection, defaultValue=<rgba 0 0.5 0 1>}
  n56 = ComponentMask(None=LinearInterpolate(
    A=LinearInterpolate(A=@n54.RGB, B=@n55.RGB, Alpha=@n3)
    B=VectorParameter() {group=HauntWrap, parameterName=MechStarsPanSpeedandDirection, defaultValue=<rgba 0 0.5 0 1>}.RGB
    Alpha=@n4
   )) {g=true, r=true}
  n57 = MaterialFunctionCall(Seed=Time(), Vec2=Ceil(None=Multiply(A=@n52) {constB=10})) {materialFunction=&MF_Hash}
  n58 = Subtract(
   A=TextureSample(UVs=Panner(Coordinate=@n53, Speed=@n56) {speedY=0.5}) {samplerSource=SSM_Wrap_WorldGroupSettings, texture=/Game/Characters/Player/Male/Medium/Bodies/M_MED_Haunt/Textures/T_StarPattern}.R
   B=ComponentMask(None=ConstantBiasScale(None=@n57.Vec2)) {r=true}
  )
  n59 = DotProduct(A=CameraVectorWS(), B=VertexNormalWS())
  n60 = Power(Base=Subtract(A=ConstantBiasScale(None=@n59), B=ScalarParameter() {defaultValue=0.5, group=HauntWrap, parameterName=Falloff}))
  n61 = Multiply(A=Multiply(A=OneMinus(None=@n27), B=Clamp(None=@n58)), B=Clamp(None=Multiply(A=@n60, B=@n41)))
  n62 = VectorParameter() {group=HauntWrap, parameterName=StarColor, defaultValue=<rgba 1 1 1 1>}
  n63 = Add(A=@n49, B=Multiply(A=Multiply(A=@n61, B=@n62.RGB), B=ScalarParameter() {defaultValue=10, group=HauntWrap, parameterName=StarGlowStrength}))
  n64 = Multiply(A=Clamp(None=@n41), B=@n60)
  n65 = Multiply(A=@n64, B=VectorParameter() {group=HauntWrap, parameterName=EdgeGlowColor, defaultValue=<rgba 0.013 0.0382 0.0782 1>}.RGB)
  n66 = Subtract(A=Power(Base=Subtract(A=@n59) {constB=0.25}), B=Multiply(A=@n41) {constB=0.25})
  n67 = Clamp(None=@n41)
  n68 = ComponentMask(None=VectorParameter() {group=HauntWrap, parameterName=LightningColor, defaultValue=<rgba 0 0.4603 1 1>}.RGB) {b=true, g=true, r=true}
  n69 = LinearInterpolate(
   A=ScalarParameter() {defaultValue=1.5, group=HauntWrap, parameterName=WeaponThunderTiling}
   B=ScalarParameter() {defaultValue=1.5, group=HauntWrap, parameterName=VehicleThunderTiling}
   Alpha=@n3
  )
  n70 = ScalarParameter() {defaultValue=1.5, group=HauntWrap, parameterName=MechThunderTiling}
  n71 = Add(A=Multiply(A=@n31, B=LinearInterpolate(A=@n69, B=@n70, Alpha=@n4)), B=@n38)
  n72 = ScalarParameter() {defaultValue=4, group=HauntWrap, parameterName=WeaponThunderDistortNoise}
  n73 = ScalarParameter() {defaultValue=4, group=HauntWrap, parameterName=VehicleThunderDistortNoise}
  n74 = Multiply(
   A=@n71
   B=LinearInterpolate(
    A=LinearInterpolate(A=@n72, B=@n73, Alpha=@n3)
    B=ScalarParameter() {defaultValue=4, group=HauntWrap, parameterName=MechThunderDistortNoise}
    Alpha=@n4
   )
  )
  n75 = VectorParameter() {group=HauntWrap, parameterName=WeaponThunderDistortioPanDirectionandSpeed, defaultValue=<rgba 0.8 0 0 1>}
  n76 = VectorParameter() {group=HauntWrap, parameterName=VehicleThunderDistortioPanDirectionandSpeed, defaultValue=<rgba 0.8 0 0 1>}
  n77 = ComponentMask(None=LinearInterpolate(
    A=LinearInterpolate(A=@n75.RGB, B=@n76.RGB, Alpha=@n3)
    B=VectorParameter() {group=HauntWrap, parameterName=MechThunderDistortioPanDirectionandSpeed, defaultValue=<rgba 0.8 0 0 1>}.RGB
    Alpha=@n4
   )) {g=true, r=true}
  n78 = ComponentMask(None=TextureSample(UVs=Panner(Coordinate=@n74, Speed=@n77) {speedX=0.5, speedY=2}) {samplerSource=SSM_Wrap_WorldGroupSettings, texture=/Game/Global/Textures/Default/Noise/T_Noise_Cloud}.RGB) {g=true, r=true}
  n79 = Multiply(
   A=Subtract(A=Multiply(A=@n78) {constB=2})
   B=ScalarParameter() {defaultValue=0.02, group=HauntWrap, parameterName=ThunderMAcroDistortion strength}
  )
  n80 = LinearInterpolate(
   A=LinearInterpolate(
    A=VectorParameter() {group=HauntWrap, parameterName=WaaponThunderPanDirectionandSpeed, defaultValue=<rgba 0.8 0 0 1>}.RGB
    B=VectorParameter() {group=HauntWrap, parameterName=VehicleThunderPanDirectionandSpeed, defaultValue=<rgba 0.8 0 0 1>}.RGB
    Alpha=@n3
   )
   B=VectorParameter() {group=HauntWrap, parameterName=MechThunderPanDirectionandSpeed, defaultValue=<rgba 0.8 0 0 1>}.RGB
   Alpha=@n4
  )
  n81 = TextureSample(UVs=Panner(Coordinate=Add(A=@n71, B=@n79), Speed=ComponentMask(None=@n80) {g=true, r=true}) {speedX=0.8}) {
   samplerSource=SSM_Wrap_WorldGroupSettings
   samplerType=SAMPLERTYPE_Masks
   texture=/Game/Weapons/WeaponSkins/Wraps/Textures/PlasmaCharge/PlasmaCharge_M
  }
  n82 = ScalarParameter() {defaultValue=2, group=HauntWrap, parameterName=ThunderBrightness}
  n83 = ScalarParameter() {defaultValue=10, group=HauntWrap, parameterName=CoreThunderBrightnes}
  n84 = Power(Base=Multiply(A=ComponentMask(None=@n81.RGB) {g=true}, B=LinearInterpolate(A=@n82, B=@n83, Alpha=@n27)) {constB=10})
  n85 = Add(A=Multiply(A=Add(A=@n63, B=@n65), B=Clamp(None=@n66) {minDefault=1}), B=Multiply(A=Add(A=@n27, B=@n67), B=Multiply(A=@n68, B=@n84)))
  n86 = LinearInterpolate(
   A=LinearInterpolate(
    A=ScalarParameter() {defaultValue=1, group=HauntWrap, parameterName=WeaponNebulaTiling}
    B=ScalarParameter() {defaultValue=1, group=HauntWrap, parameterName=VehicleNebulaTiling}
    Alpha=@n3
   )
   B=ScalarParameter() {defaultValue=1, group=HauntWrap, parameterName=MechNebulaTiling}
   Alpha=@n4
  )
  n87 = LinearInterpolate(
   A=LinearInterpolate(
    A=VectorParameter() {group=HauntWrap, parameterName=WeaponNebulaPanSpeed, defaultValue=<rgba 0 -0.2 0 1>}.RGB
    B=VectorParameter() {group=HauntWrap, parameterName=VehicleNebulaPanSpeed, defaultValue=<rgba 0 -0.2 0 1>}.RGB
    Alpha=@n3
   )
   B=VectorParameter() {group=HauntWrap, parameterName=MechNebulaPanSpeed, defaultValue=<rgba 0 -0.2 0 1>}.RGB
   Alpha=@n4
  )
  n88 = Add(A=Panner(Coordinate=Multiply(A=@n31, B=@n86), Speed=ComponentMask(None=@n87) {g=true, r=true}) {speedY=-0.8}, B=@n38)
  n89 = ComponentMask(None=@n88) {r=true}
  n90 = ComponentMask(None=@n88) {g=true}
  n91 = AppendVector(A=Add(A=@n89) {constB=0.025}, B=@n90)
  n92 = AppendVector(A=Add(A=@n89) {constB=0.05}, B=@n90)
  n93 = Multiply(
   A=@n41
   B=MaterialFunctionCall(
    X=TextureSample(UVs=@n88) {samplerSource=SSM_Wrap_WorldGroupSettings, texture=/Game/Global/Textures/Default/Caustics/T_Turbulence_Seamless}.R
    Y=TextureSample(UVs=@n91) {samplerSource=SSM_Wrap_WorldGroupSettings, texture=&T_Turbulence_Seamless}.R
    Z=TextureSample(UVs=@n92) {samplerSource=SSM_Wrap_WorldGroupSettings, texture=&T_Turbulence_Seamless}.R
   ) {materialFunction=&MakeFloat3}.Result
  )
  n94 = ScalarParameter() {defaultValue=0.5, group=HauntWrap, parameterName=NebulaGlowStrength}
  n95 = GetMaterialAttributes(None=FunctionInput() {inputName=Material, inputType=FunctionInput_MaterialAttributes})
  n96 = SetMaterialAttributes!(
   MaterialAttributes=@n95.MaterialAttributes
   BaseColor=@n49
   Metallic=Constant()
   Specular=Constant() {r=1}
   Roughness=Constant() {r=0.35}
   EmissiveColor=Add(A=@n85, B=Multiply(A=@n93, B=@n94)) {constB=0}
  ) {attributeSetTypes [
    #1
    #2
    #3
    #4
    #5
   ]}
  n97 = FunctionInput(Preview=Constant() {r=1}) {
   bUsePreviewValueAsDefault=true
   inputName="Customization Mask (_CM B)"
   inputType=FunctionInput_Scalar
   sortPriority=24
   previewValue=<1 0 0 1>
  }
  FunctionOutput(None=StaticSwitchParameter(True=BlendMaterialAttributes(A=@n95.MaterialAttributes, B=@n96, Alpha=@n97), False=@n95.MaterialAttributes) {group=Customization_Wrap, parameterName=UseHauntWrap})
  MaterialFunctionCall() {materialFunction=&CameraDirectionVector}
  n98 = CameraPositionWS()
  n99 = ObjectPositionWS()
  n100 = TextureSample(UVs=@n39) {
   samplerSource=SSM_Wrap_WorldGroupSettings
   samplerType=SAMPLERTYPE_Normal
   texture=/Game/Characters/Player/Male/Medium/Bodies/M_MED_Haunt/Textures/T_Thick_Smoke_05_N
  }
  n101 = DotProduct(A=Normalize(VectorInput=Subtract(A=@n98, B=@n99)), B=Transform(None=@n100.RGB))
  n102 = Clamp(None=Subtract(A=ConstantBiasScale(None=@n101), B=ScalarParameter() {group=HauntWrap, parameterName=VolumeShadowFalloff}) {constB=0.25})
  Multiply(A=@n85, B=ShadingPathSwitch(Default=@n102, Deferred=@n102, Mobile=Constant() {r=1}) {desc=Migrated from FeatureLevelSwitch})
  n103 = TextureSample(UVs=@n88) {
   samplerSource=SSM_Wrap_WorldGroupSettings
   texture=/Game/Characters/Player/Male/Medium/Bodies/M_Med_Soldier_04/Skins/Galaxy/Textures/T_FN_Nebula
  }
  n104 = Constant() {r=0.1}
  Multiply(
   A=Multiply(A=Desaturation(None=@n103.RGB, Fraction=@n104), B=@n64)
   B=VectorParameter() {group=HauntWrap, parameterName=Nerbula_Color, defaultValue=<rgba 0.5336 0.7712 1 1>}.RGB
  )
  NamedRerouteDeclaration(None=FunctionInput() {bUsePreviewValueAsDefault=true, inputName=IsCharacter, inputType=FunctionInput_Scalar}) {name=IsCharacter, nodeColor=<rgba 0 0.9253 1 1>}
  n105 = FunctionInput() {bUsePreviewValueAsDefault=true, inputName=IsStatsB, inputType=FunctionInput_Vector2}
  NamedRerouteDeclaration(None=ComponentMask(None=@n105) {r=true}) {name=IsBackpack, nodeColor=<rgba 0 0.9253 1 1>}
  NamedRerouteDeclaration(None=ComponentMask(None=@n105) {g=true}) {name=IsPickaxe, nodeColor=<rgba 0 0.9253 1 1>}
  ObjectPositionWS()
  n106 = VertexInterpolator(VS=FunctionInput(Preview=PreSkinnedPosition()) {bUsePreviewValueAsDefault=true, inputName=PreskinnedPosition})
  n107 = FunctionInput(Preview=PreSkinnedLocalBounds().Max) {bUsePreviewValueAsDefault=true, inputName=PreskinnedBounds}
  Subtract(A=Divide(A=ComponentMask(None=@n106.PS) {g=true}, B=ComponentMask(None=@n107) {g=true}), B=@n25)
 ]}
