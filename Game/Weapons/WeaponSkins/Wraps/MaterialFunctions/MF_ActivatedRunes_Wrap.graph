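# Weapon wrap material function for activated runes, enabled by a static switch and blended over the bypass material through the customization mask.
# Projects color, normal and mask textures top/side within wrap bounds, with per-item-type offsets and scales; mobile skips projected normals.
# The mask shapes and warps an animated gradient that drives curve-atlas emissive color.
# That glow fades over a set duration from a Blueprint-supplied game time, making it reactive.
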MaterialFunction /Game/Weapons/WeaponSkins/Wraps/MaterialFunctions/MF_ActivatedRunes_Wrap
graph {nodes [
  n1 = GetMaterialAttributes(None=FunctionInput() {inputName=Material_Bypass, inputType=FunctionInput_MaterialAttributes, sortPriority=1}) {attributeGetTypes [#1, #2, #3]}
  n2 = TextureObjectParameter() {
   group=TopSideProjection
   parameterName=TS_Normal_01
   samplerType=SAMPLERTYPE_Normal
   sortPriority=16
   texture=&FlatNormalBlue
  }
  n3 = Constant() {r=1}
  n4 = FunctionInput(Preview=Constant()) {bUsePreviewValueAsDefault=true, inputName=isMech, inputType=FunctionInput_Scalar, sortPriority=4}
  n5 = LinearInterpolate(A=@n3, B=Constant() {r=2}, Alpha=@n4)
  n6 = FunctionInput(Preview=Constant()) {inputName=isVehicle, inputType=FunctionInput_Scalar, sortPriority=5}
  n7 = LinearInterpolate(
   A=LinearInterpolate(A=Constant3Vector() {constant=<rgba 0 0 -0.25 1>}, B=Constant3Vector(), Alpha=@n6)
   B=Constant3Vector() {constant=<rgba 0.2502 0.2502 0 0>}
   Alpha=@n4
  )
  n8 = FunctionInput(Preview=VertexInterpolator(VS=PreSkinnedNormal()).PS) {bUsePreviewValueAsDefault=true, inputName=Pre-Skinned Local Normal, sortPriority=10}
  n9 = VectorParameter() {group=TopSideProjection, parameterName=TS_Offset_Weapon, sortPriority=-4, defaultValue=<rgba 0.3325 0.3278 0 1>}
  n10 = VectorParameter() {group=TopSideProjection, parameterName=TS_Offset_Vehicle, sortPriority=-3, defaultValue=<rgba 0.3325 0.3278 0 1>}
  n11 = LinearInterpolate(A=@n9.RGB, B=@n10.RGB, Alpha=@n6)
  n12 = VectorParameter() {group=TopSideProjection, parameterName=TS_Offset_Mech, sortPriority=-2, defaultValue=<rgba 0.3325 0.3278 0 1>}
  n13 = VectorParameter() {group=TopSideProjection, parameterName=TS_Offset_Character, sortPriority=-2, defaultValue=<rgba 0 0 0 1>}
  n14 = VectorParameter() {group=TopSideProjection, parameterName=TS_Offset_Pickaxe, sortPriority=-2, defaultValue=<rgba 0 0 0 1>}
  n15 = LinearInterpolate(
   A=LinearInterpolate(A=LinearInterpolate(A=@n11, B=@n12.RGB, Alpha=@n4), B=@n13.RGB, Alpha=NamedRerouteUsage())
   B=@n14.RGB
   Alpha=NamedRerouteUsage()
  )
  n16 = VectorParameter() {group=TopSideProjection, parameterName=TS_Offset_Backpack, sortPriority=-2, defaultValue=<rgba 0 0 0 1>}
  n17 = NamedRerouteUsage()
  n18 = PreSkinnedPosition()
  n19 = MaterialFunctionCall!(
   is Vehicle=@n6
   Object Bounds Max Adjustment=FunctionInput() {bUsePreviewValueAsDefault=true, inputName=Object Bounds Max Adjustment, sortPriority=13}
   Object Bounds Min Adjustment=FunctionInput() {bUsePreviewValueAsDefault=true, inputName=Object Bounds Min Adjustment, sortPriority=12}
   Object Bounds Center Adjustment=FunctionInput() {bUsePreviewValueAsDefault=true, inputName=Object Bounds Center Adjustment, sortPriority=14}
   isMech=@n4
   Side Project Vehicle
   Local Position=FunctionInput(Preview=VertexInterpolator(VS=@n18).PS) {bUsePreviewValueAsDefault=true, inputName=Local Position, sortPriority=11}
  ) {materialFunction=&/MF_WrapBounds}
  n20 = VectorParameter() {group=TopSideProjection, parameterName=TS_Scale_Weapon, sortPriority=-4, defaultValue=<rgba 1 1 1 1>}
  n21 = VectorParameter() {group=TopSideProjection, parameterName=TS_Scale_Vehicle, sortPriority=-3, defaultValue=<rgba 1 1 1 1>}
  n22 = VectorParameter() {group=TopSideProjection, parameterName=TS_Scale_Mech, sortPriority=-2, defaultValue=<rgba 1 1 1 1>}
  n23 = VectorParameter() {group=TopSideProjection, parameterName=TS_Scale_Character, sortPriority=-2, defaultValue=<rgba 1 1 1 1>}
  n24 = LinearInterpolate(A=LinearInterpolate(A=LinearInterpolate(A=@n20.RGB, B=@n21.RGB, Alpha=@n6), B=@n22.RGB, Alpha=@n4), B=@n23.RGB, Alpha=NamedRerouteUsage())
  n25 = VectorParameter() {group=TopSideProjection, parameterName=TS_Scale_Backpack, sortPriority=-2, defaultValue=<rgba 1 1 1 1>}
  n26 = NamedRerouteUsage()
  n27 = VectorParameter() {group=TopSideProjection, parameterName=TS_Scale_Pickaxe, sortPriority=-2, defaultValue=<rgba 1 1 1 1>}
  n28 = Multiply(
   A=Add(A=@n19.Result, B=LinearInterpolate(A=@n15, B=@n16.RGB, Alpha=@n17))
   B=LinearInterpolate(A=LinearInterpolate(A=@n24, B=@n25.RGB, Alpha=@n26), B=@n27.RGB, Alpha=NamedRerouteUsage())
  )
  n29 = LinearInterpolate(
   A=@n1.Normal
   B=Add(
    A=@n1.Normal
    B=MaterialFunctionCall!(
     Texture_Normals=@n2
     Rotation_Axis=@n7
     Caculated_Bounds=@n28
     Pre_Skinned_Normals=@n8
     Blend_Contrast=@n5
     Fill_alternate_Axis_Opacity=@n3
     Is_Vehicle=@n6
     Is_Mech=@n4
    ) {materialFunction=&/MF_TopsSideProjection_Normals}.Add_Surface_Normal
   )
   Alpha=Constant() {r=1}
  )
  n30 = TextureObjectParameter() {
   group=TopSideProjection
   parameterName=TS_Mask_01
   samplerType=SAMPLERTYPE_Masks
   sortPriority=20
   texture=&T_BlackMask
  }
  n31 = MaterialFunctionCall!(
   IsCharacter
   IsStatsB
   Texture=@n30
   Rotation_Axis=@n7
   Caculated_Bounds=@n28
   Pre_Skinned_Normals=@n8
   Blend_Contrast=@n5
   Fill_alternate_Axis_Opacity=@n3
   "Fill_alternate_Axis_Color_A=Roughness"=Constant4Vector().RGBA
   Is_Vehicle=@n6
   Is_Mech=@n4
   Enable_Fill_Z_Weapon
   Fill_Z_Weapon_opacity
   Fill_Z_Weapon_Color
  ) {materialFunction=&/MF_TopsSideProjection_Base}
  n32 = VectorParameter() {group=TopSideProjection_Emissive, parameterName=TS_Emissive01_OffsetWeapon, sortPriority=25, defaultValue=<rgba 0 -1 0 0>}
  n33 = ScalarParameter() {defaultValue=0.5, group=TopSideProjection_Emissive, parameterName="TS_Emissive01_FalloffWeapon (negative inverts)", sortPriority=23}
  n34 = VectorParameter() {group=TopSideProjection_Emissive, parameterName=TS_Emissive01_OffsetVehicle, sortPriority=26, defaultValue=<rgba -1 0 0 0>}
  n35 = ScalarParameter() {defaultValue=0.5, group=TopSideProjection_Emissive, parameterName="TS_Emissive01_FalloffVehicle ", sortPriority=24}
  n36 = VectorParameter() {group=TopSideProjection_Emissive, parameterName=TS_Emissive01_OffsetMech, sortPriority=27, defaultValue=<rgba 0 0 -1 0>}
  n37 = ScalarParameter() {defaultValue=0.5, group=TopSideProjection_Emissive, parameterName="TS_Emissive01_FalloffMech ", sortPriority=25}
  n38 = StaticBoolParameter() {group=TopSideProjection_Emissive, parameterName=TS_Emissive01_Invert_Gradent, sortPriority=11}
  n39 = VectorParameter() {group=TopSideProjection_Emissive, parameterName=TS_Emissive01_Spread_XY, sortPriority=21, defaultValue=<rgba 1 1 1 0>}
  n40 = VectorParameter() {group=TopSideProjection_Emissive, parameterName=TS_Emissive01_Spread_Mech_XY, sortPriority=20, defaultValue=<rgba 1 1 1 0>}
  n41 = ScalarParameter() {defaultValue=1, group=TopSideProjection_Emissive, parameterName=TS_Warp_SpeedModifier, sortPriority=38}
  n42 = Multiply(A=LinearInterpolate(A=@n39.RGB, B=@n40.RGB, Alpha=@n4), B=@n41)
  n43 = VectorParameter() {group=TopSideProjection_Emissive, parameterName=TS_Emissive01_Animated_WeaponXY, sortPriority=30}
  n44 = VectorParameter() {group=TopSideProjection_Emissive, parameterName=TS_Emissive01_Animated_VehicleXY, sortPriority=31}
  n45 = Multiply(
   A=LinearInterpolate(
    A=LinearInterpolate(A=@n43.RGB, B=@n44.RGB, Alpha=@n6)
    B=VectorParameter() {group=TopSideProjection_Emissive, parameterName=TS_Emissive01_Animated_MechXY}.RGB
    Alpha=@n4
   )
   B=@n41
  )
  n46 = Multiply(
   A=ComponentMask(None=@n31.Combined) {a=true}
   B=ScalarParameter() {group=TopSideProjection_Emissive, parameterName=TS_WarpEmissive01_Warp_Amount, sortPriority=38}
  )
  n47 = Constant() {r=1}
  n48 = FunctionInput(Preview=Constant3Vector() {constant=<rgba 1 0 0 0>}) {bUsePreviewValueAsDefault=true, inputName=ForwardAxis}
  n49 = TextureObjectParameter() {
   group=TopSideProjection_Emissive
   parameterName=Ts_Emissive01Gradent_Texture
   samplerType=SAMPLERTYPE_Masks
   sortPriority=12
   texture=/Game/Global/Textures/Default/Masks_and_Gradients/T_VertGradient_MASK
  }
  n50 = Abs(None=ComponentMask(None=@n8) {b=true})
  n51 = Constant() {r=1}
  n52 = Abs(None=ComponentMask(None=@n8) {g=true})
  n53 = Saturate(None=Subtract(
    A=MaterialFunctionCall(In=@n50, Contrast=@n51) {materialFunction=&CheapContrast}.Result
    B=MaterialFunctionCall(In=@n52, Contrast=@n51) {materialFunction=&CheapContrast}.Result
   ))
  n54 = MaterialFunctionCall(In=Abs(None=ComponentMask(None=@n8) {r=true}), Contrast=@n51) {materialFunction=&CheapContrast}
  n55 = LinearInterpolate(A=Saturate(None=Add(A=@n53, B=@n54.Result)), B=@n51, Alpha=@n6)
  n56 = Constant()
  n57 = Multiply(
   A=TextureSample(
    UVs=MaterialFunctionCall!(
     ForwardAxis=@n48
     Bounds=@n19.Result
     Pre-Skinned Local Position
     isVehicle=@n6
     isMech=@n4
     Side Project Vehicle
     WeaponOffset=@n32.RGB
     WeaponFalloff=@n33
     VehicleOffset=@n34.RGB
     VehicleFalloff=@n35
     MechOffset=@n36.RGB
     MechFalloff=@n37
     Invert_Gradent=@n38
     UV_Spread_XY=@n42
     UV_Animated_XY=@n45
     Warp_UVs=@n46
     Rotate_Mech=@n47
     Rotate_Vehicle_Side
     Rotate_Vehicle=@n47
    ) {materialFunction=&/MF_Wrap_Gradent_Creation}.Uvs
    Tex=@n49
   ).R
   B=Saturate(None=Add(A=@n55, B=@n56))
  )
  n58 = Multiply(A=ComponentMask(None=@n31.Combined) {r=true}, B=@n57)
  n59 = Saturate(None=Multiply(A=@n58, B=Add(A=@n58, B=@n57)))
  n60 = Saturate(None=Add(A=@n59, B=Constant() {r=1}))
  n61 = Time()
  n62 = Frac(None=Add(A=Multiply(B=@n61), B=Saturate(None=@n59)))
  n63 = CurveAtlasRowParameter(CurveTime=@n62) {
   atlas=/Game/Weapons/WeaponSkins/Wraps/Textures/CruveAtlas_Wraps_01
   curve=/Game/Weapons/WeaponSkins/Wraps/Textures/C_GradSimple
   group=TopSideProjection_Emissive
   parameterName=TS_Emissive01_Curve
   sortPriority=3
  }
  n64 = MaterialFunctionCall!(
   Index=ScalarParameter() {desc=look under 00_ColorCurve, group=TopSideProjection_Emissive, parameterName=TS_Emissive01_Curve_Index, sortPriority=4}
   CCinput=@n63
   CCinput_Alpha=@n63
   UV=@n62
  ) {materialFunction=&/MF_Global_CC_Tex}
  n65 = Multiply(A=@n64.3V, B=Multiply(A=@n64.Alpha, B=ScalarParameter() {defaultValue=1, parameterName=TS_CurveBrightness, sortPriority=4}))
  n66 = Subtract(A=Time(), B=ScalarParameter() {group=TopSideProjection_Reactive, parameterName=TS_GameTimeFromBP})
  n67 = ScalarParameter() {defaultValue=1, group=TopSideProjection_Reactive, parameterName=TS_AnimationDuration}
  n68 = Power(
   Base=Saturate(None=Divide(A=@n66, B=@n67))
   Exp=ScalarParameter() {defaultValue=0.5, group=TopSideProjection_Reactive, parameterName=TS_Curve_Amount}
  )
  n69 = TextureObjectParameter() {group=TopSideProjection, parameterName=TS_BaseColor_01, sortPriority=11, texture=&T_BlacksRGB}
  n70 = VectorParameter() {group=TopSideProjection, parameterName="TS_Fill_alternate_Axis_Color_A=Roughness 01", sortPriority=3}
  n71 = MaterialFunctionCall(
   Texture=@n69
   Rotation_Axis=@n7
   Caculated_Bounds=@n28
   Pre_Skinned_Normals=@n8
   Blend_Contrast=@n5
   Fill_alternate_Axis_Opacity=@n3
   "Fill_alternate_Axis_Color_A=Roughness"=AppendVector(A=@n70.RGB, B=@n70.RGB)
   Is_Vehicle=@n6
   Is_Mech=@n4
  ) {materialFunction=&MF_TopsSideProjection_Base}
  n72 = SetMaterialAttributes!(
   MaterialAttributes=@n1.MaterialAttributes
   BaseColor=ComponentMask(None=@n71.Combined) {b=true, g=true, r=true}
   Roughness=LinearInterpolate(
    A=ScalarParameter() {group=TopSideProjection, parameterName=TS_Roughness_Min, sortPriority=13}
    B=ScalarParameter() {defaultValue=1, group=TopSideProjection, parameterName=TS_Roughness_Max, sortPriority=14}
    Alpha=ComponentMask(None=@n71.Combined) {a=true}
   )
   Normal=ShadingPathSwitch(Default=@n29, Deferred=@n29, Mobile=@n1.Normal) {desc=Migrated from FeatureLevelSwitch}
   Specular=ScalarParameter() {defaultValue=0.5, group=TopSideProjection, parameterName=TS_Specularty, sortPriority=12}
   Metallic=ScalarParameter() {group=TopSideProjection, parameterName=TS_Metallic, sortPriority=26}
   EmissiveColor=LinearInterpolate(B=Multiply(A=@n60, B=@n65), Alpha=OneMinus(None=@n68) {desc=inverts curve from 0-1 to 1-0})
  ) {attributeSetTypes [
    #1
    #2
    #3
    #4
    #5
    #6
   ]}
  n73 = StaticSwitchParameter(
   True=BlendMaterialAttributes(
    A=@n1.MaterialAttributes
    B=@n72
    Alpha=FunctionInput() {bUsePreviewValueAsDefault=true, inputName="Customization Mask (_CM B)", inputType=FunctionInput_Scalar, sortPriority=24}
   )
   False=FunctionInput() {inputName=Material, inputType=FunctionInput_MaterialAttributes}
  ) {desc=Topside Projection - Reactive - Emissive X+X texture warping, group=Customization_Wrap, parameterName=Wrap_Activated_Runes}
  FunctionOutput(None=@n73)
  LinearInterpolate(A=@n55, Alpha=@n4)
  NamedRerouteDeclaration(None=FunctionInput() {bUsePreviewValueAsDefault=true, inputName=isCharacter, inputType=FunctionInput_Scalar}) {name=IsCharacter, nodeColor=<rgba 0.3125 0 1 1>}
  n74 = NamedRerouteDeclaration(None=FunctionInput() {bUsePreviewValueAsDefault=true, inputName=IsStatsB, inputType=FunctionInput_Vector2, sortPriority=78}) {name=Is_StatsB, nodeColor=<rgba 0 0.9253 1 1>}
  NamedRerouteDeclaration(None=ComponentMask(None=@n74) {r=true}) {name=IsBackpack, nodeColor=<rgba 0.2656 0 1 1>}
  NamedRerouteDeclaration(None=ComponentMask(None=@n74) {g=true}) {name=IsPickaxe, nodeColor=<rgba 1 0 0.5625 1>}
  PreSkinnedLocalBounds()
  ScalarParameter() {defaultValue=1, group=TopSideProjection_Reactive, parameterName=TurnOnEmissive_ForSequencer}
 ]}
