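# Banner wrap material function that stamps a banner icon and a primary-color triangle pattern onto weapons, vehicles and mechs.
# Weapons project the icon from pre-skinned bounds, mirrored by side; vehicles and mechs project it per axis from local position and normal.
# A height-lerped world-aligned triangle texture forms a gradient mask, and an offset icon sample darkens the color.
# The UseBannerWrap switch blends this metallic layer over the input attributes, scaled by CMask.
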MaterialFunction /Game/Weapons/WeaponSkins/Wraps/MaterialFunctions/MF_BannerWrap
graph {nodes [
  n1 = VertexInterpolator(VS=PreSkinnedPosition())
  n2 = PreSkinnedLocalBounds()
  n3 = FunctionInput(Preview=Divide(A=Subtract(A=@n1.PS, B=@n2.Min), B=Subtract(A=@n2.Max, B=@n2.Max))) {bUsePreviewValueAsDefault=true, inputName=Pre-SkinBounds}
  n4 = ComponentMask(None=VectorParameter() {group=Banner, parameterName=BannerCharacter_Icon_offset, defaultValue=<rgba 0 0 0 1>}.RGB) {b=true, g=true, r=true}
  n5 = ComponentMask(None=VectorParameter() {group=Banner, parameterName=Banner_Backpack_Icon_offset, defaultValue=<rgba 0 0 0 1>}.RGB) {b=true, g=true, r=true}
  n6 = NamedRerouteUsage()
  n7 = LinearInterpolate(A=@n3, B=Add(A=LinearInterpolate(A=@n4, B=@n5, Alpha=@n6), B=@n3), Alpha=NamedRerouteUsage())
  ComponentMask(None=@n7) {g=true}
  Constant3Vector() {constant=<rgba 0.6654 0 0 0>}
  Constant3Vector() {constant=<rgba 0.8308 0.552 0 0>}
  Constant3Vector() {constant=<rgba 0 0.8148 0 0>}
  Constant3Vector() {constant=<rgba 0 0.8148 0.8148 0>}
  Constant3Vector() {constant=<rgba 0.011 0.011 0.011 0>}
  Constant3Vector() {constant=<rgba 0.6038 0.6038 0.6038 0>}
  Constant3Vector() {constant=<rgba 0.0908 0 0 0>}
  Constant3Vector() {constant=<rgba 1 1 0.3005 0>}
  Constant3Vector() {constant=<rgba 0 0 0.0908 0>}
  Constant() {r=20}
  Constant() {r=22}
  Constant() {r=3.14}
  Constant() {r=3.14}
  Constant() {r=0.5}
  Constant() {r=0.1}
  Constant() {r=0.5}
  Constant() {r=0.1}
  Constant() {r=-2.2}
  Constant() {r=0.5}
  Constant() {r=0.7}
  Constant() {r=-6.28}
  Constant() {r=0.7}
  Constant() {r=-3.14}
  Constant() {r=12.57}
  Constant() {r=-3.14}
  Constant() {r=3.14}
  Constant() {r=18.8}
  Constant() {r=15.6094}
  Constant() {r=1}
  Constant() {r=0.25}
  Constant() {r=15}
  n8 = FunctionInput() {inputType=FunctionInput_MaterialAttributes}
  n9 = MaterialFunctionCall(Color=@n7) {materialFunction=&SplitComponents}
  n10 = Divide(A=Add(A=@n9.G, B=Constant() {r=0.2}), B=Constant() {r=0.8})
  n11 = Divide(A=Add(A=@n9.R, B=Constant() {r=-0.2}), B=Constant() {r=0.5})
  n12 = FunctionInput(Preview=Constant()) {bUsePreviewValueAsDefault=true, inputName=IsVehicle, inputType=FunctionInput_Scalar, sortPriority=8}
  n13 = Add(A=@n9.B, B=Constant() {r=0.01})
  n14 = Constant() {r=0.5}
  n15 = FunctionInput(Preview=Constant()) {bUsePreviewValueAsDefault=true, inputName=IsMech, inputType=FunctionInput_Scalar, sortPriority=9}
  n16 = LinearInterpolate(
   A=LinearInterpolate(A=Saturate(None=@n10), B=Saturate(None=@n11), Alpha=@n12) {desc="B=Vehicle"}
   B=Saturate(None=Divide(A=@n13, B=@n14))
   Alpha=@n15
  ) {desc="B=Mech"}
  n17 = Max(A=@n12, B=@n15)
  n18 = FunctionInput(Preview=VertexInterpolator(VS=PreSkinnedNormal()).PS) {bUsePreviewValueAsDefault=true, inputName=Pre-SkinLocalNormal}
  n19 = FunctionInput(Preview=VertexInterpolator(VS=PreSkinnedPosition()).PS) {bUsePreviewValueAsDefault=true, inputName=Pre-SkinLocalPos}
  n20 = MaterialFunctionCall(
   TextureObject=TextureObject() {samplerType=SAMPLERTYPE_LinearColor, texture=/Game/Global/Textures/Default/Masks_and_Gradients/T_TrianglePattern_VectorDisplacement}
   TextureSize=LinearInterpolate(A=Constant() {r=14}, B=Constant() {r=75}, Alpha=@n17) {desc="B=Vehicle"}
   WorldPosition=@n19
   World Space Normal=@n18
   ProjectionTransitionContrast=Constant() {r=100}
  ) {materialFunction=&WorldAlignedTexture}
  n21 = Constant() {r=1}
  n22 = Saturate(None=Multiply(
    A=MaterialFunctionCall(Transition Phase=@n16, Height Texture=@n20.XYZ Texture, Contrast=@n21) {materialFunction=&HeightLerp}.Alpha
    B=Constant() {r=3}
   ))
  n23 = LinearInterpolate(
   A=ComponentMask(None=VectorParameter() {group=Banner, parameterName=BannerCharacter_Icon_offset, defaultValue=<rgba 0 0 0 1>}.RGB) {g=true, r=true}
   B=ComponentMask(None=VectorParameter() {group=Banner, parameterName=Banner_Backpack_Icon_offset, defaultValue=<rgba 0 0 0 1>}.RGB) {g=true, r=true}
   Alpha=NamedRerouteUsage()
  )
  n24 = LinearInterpolate(
   A=ScalarParameter() {defaultValue=22, group=Banner_Vehicle, parameterName=BannerScale1_Vehicle}
   B=ScalarParameter() {defaultValue=20, group=Banner_Mech, parameterName=BannerScale1_Mech}
   Alpha=@n15
  )
  n25 = MaterialFunctionCall(Float3=Divide(A=@n19, B=Multiply(A=@n24) {constB=-1})) {materialFunction=&BreakOutFloat3Components}
  n26 = AppendVector(A=@n25.R, B=@n25.R)
  n27 = Constant() {r=100}
  n28 = MaterialFunctionCall(Float3=Saturate(None=Multiply(A=@n18, B=@n27))) {materialFunction=&BreakOutFloat3Components}
  n29 = LinearInterpolate(A=@n26, B=Multiply(A=@n26, B=Constant2Vector() {g=1, r=-1}) {constB=-1}, Alpha=@n28.G)
  n30 = NamedRerouteUsage()
  n31 = LinearInterpolate(A=@n29, B=Add(A=@n29, B=@n23), Alpha=@n30)
  n32 = StaticBool()
  n33 = TextureObjectParameter() {parameterName=Banner_Texture, texture=/Game/UI/Foundation/Textures/Banner/Standard/T-Banners-Icons-Chevrons-L}
  n34 = ScalarParameter() {defaultValue=0.9, group=Banner_Vehicle, parameterName=BannerScale2_Vehicle}
  n35 = AppendVector(A=@n25.G, B=@n25.G)
  n36 = LinearInterpolate(A=Multiply(A=@n35, B=Constant2Vector() {g=1, r=-1}), B=@n35, Alpha=@n28.R)
  n37 = LinearInterpolate(A=@n36, B=Add(A=@n36, B=@n23), Alpha=@n30)
  n38 = Add(A=Abs(None=@n18), B=Constant() {r=-0.5})
  n39 = Constant() {r=100}
  n40 = MaterialFunctionCall(Float3=Saturate(None=Multiply(A=@n38, B=@n39))) {materialFunction=&BreakOutFloat3Components}
  n41 = AppendVector(A=@n25.R, B=@n25.R)
  n42 = LinearInterpolate(A=Multiply(A=@n41, B=Constant2Vector() {g=-1, r=1}) {constB=-1}, B=@n41, Alpha=@n28.B)
  n43 = Add(A=@n42, B=@n23)
  n44 = ScalarParameter() {defaultValue=6.33, group=Banner_Vehicle, parameterName=RotateBanner_Top_Vehicle}
  n45 = ScalarParameter() {defaultValue=12.5, group=Banner_Mech, parameterName=RotateBanner_Top_Mech}
  n46 = LinearInterpolate(
   A=LinearInterpolate(
    A=MaterialFunctionCall!(In-UVs=@n31, Rotation, UseRotation=@n32, OffsetLoc, Scale2=@n34, Texture=@n33) {materialFunction=&/MF_BannerWrap_CheckerMask}.Banner
    B=MaterialFunctionCall(In-UVs=@n37, UseRotation=@n32, Scale2=@n34, Texture=@n33) {materialFunction=&MF_BannerWrap_CheckerMask}.Banner
    Alpha=@n40.R
   ) {desc=Float3}
   B=MaterialFunctionCall(
    In-UVs=LinearInterpolate(A=@n42, B=@n43, Alpha=@n30)
    Rotation=LinearInterpolate(A=@n44, B=@n45, Alpha=@n15)
    UseRotation=StaticBool() {value=true}
    Scale2=@n34
    Texture=@n33
   ) {materialFunction=&MF_BannerWrap_CheckerMask}.Banner
   Alpha=@n40.B
  )
  n47 = Constant() {r=0.5}
  n48 = LinearInterpolate(
   A=VectorParameter() {group=Banner, parameterName=Banner_Weapon_Icon_offset, defaultValue=<rgba 0 0 0 1>}.RGB
   B=VectorParameter() {group=Banner, parameterName=Banner_Pickaxe_Icon_offset, defaultValue=<rgba 0 0 0 1>}.RGB
   Alpha=NamedRerouteUsage()
  )
  n49 = MaterialFunctionCall(
   UVs=ComponentMask(None=Add(A=@n7, B=@n48)) {b=true, g=true}
   Texture Scale=ScalarParameter() {defaultValue=0.5, group=Banner, parameterName=Banner_Weapon_Icon_Size}
  ) {materialFunction=&ScaleUVsByCenter}
  n50 = MaterialFunctionCall(Float2=@n49.UVs) {materialFunction=&BreakOutFloat2Components}
  n51 = Constant() {r=1.5}
  n52 = Constant() {r=0.5}
  n53 = Multiply(A=Constant2Vector() {g=-1, r=1}, B=@n49.UVs)
  n54 = Frac(None=@n53)
  n55 = ShadingPathSwitch(Default=@n53, Forward=@n54, Mobile=@n54)
  n56 = Constant2Vector() {g=-0.02}
  n57 = Add(A=@n55, B=@n56)
  n58 = OneMinus(None=@n49.UVs)
  n59 = Frac(None=@n58)
  n60 = ShadingPathSwitch(Default=@n58, Forward=@n59, Mobile=@n59)
  n61 = Add(A=@n60, B=@n56)
  n62 = ComponentMask(None=@n19) {r=true}
  n63 = Constant() {r=-1.5}
  n64 = LinearInterpolate(
   A=MaterialFunctionCall(
    X=TextureSample(UVs=@n55, Tex=@n33) {samplerSource=SSM_Wrap_WorldGroupSettings}.R
    Y=TextureSample(UVs=@n57, Tex=@n33) {samplerSource=SSM_Wrap_WorldGroupSettings}.R
   ) {materialFunction=&MakeFloat2}.Result
   B=MaterialFunctionCall(
    X=TextureSample(UVs=@n60, Tex=@n33) {samplerSource=SSM_Wrap_WorldGroupSettings}.R
    Y=TextureSample(UVs=@n61, Tex=@n33) {samplerSource=SSM_Wrap_WorldGroupSettings}.R
   ) {materialFunction=&MakeFloat2}.Result
   Alpha=Saturate(None=Add(A=@n62, B=@n63))
  )
  n65 = Constant() {r=0.5}
  n66 = MaterialFunctionCall(Float2=@n60) {materialFunction=&BreakOutFloat2Components}
  n67 = Constant() {r=0.5}
  n68 = Constant() {r=0.98}
  n69 = Multiply(
   A=@n64
   B=Multiply(
    A=SphereMask(A=@n65, B=@n66.R, Radius=@n67, Hardness=@n68) {attenuationRadius=0.5, hardnessPercent=0}
    B=SphereMask(A=@n66.G, B=@n65, Radius=@n67, Hardness=@n68) {attenuationRadius=0.5, hardnessPercent=0}
   )
  )
  n70 = ShadingPathSwitch(Default=@n64, Forward=@n69, Mobile=@n69)
  n71 = Multiply(
   A=Multiply(A=SphereMask(A=@n47, B=@n50.R, Radius=@n51) {attenuationRadius=0.5}, B=SphereMask(A=@n50.G, B=@n52) {attenuationRadius=0.5})
   B=@n70
  )
  n72 = MaterialFunctionCall(Float2=LinearInterpolate(A=@n71, B=Multiply(A=@n22, B=@n46), Alpha=@n17)) {materialFunction=&BreakOutFloat2Components}
  n73 = Add(A=Constant() {r=0.85}, B=OneMinus(None=@n72.R))
  n74 = Constant() {r=0.015}
  n75 = Constant() {r=0.6}
  n76 = VectorParameter() {group=Banner, parameterName=Banner_PrimaryColor, defaultValue=<rgba 0.6654 0 0 1>}
  n77 = MaterialFunctionCall(
   X=LinearInterpolate(A=@n74, B=@n75, Alpha=@n76.R)
   Y=LinearInterpolate(A=@n74, B=@n75, Alpha=@n76.G)
   Z=LinearInterpolate(A=@n74, B=@n75, Alpha=@n76.B)
  ) {materialFunction=&MakeFloat3}
  n78 = Multiply(A=@n20.XYZ Texture, B=Constant() {r=4})
  n79 = Saturate(None=Add(A=Saturate(None=@n78), B=Constant() {r=0.7}))
  n80 = OneMinus(None=@n72.G)
  n81 = Max(A=Multiply(A=Multiply(A=@n77.Result, B=@n79), B=Multiply(A=@n22, B=@n80)), B=Multiply(A=@n72.R, B=Constant() {r=0.9}))
  n82 = SetMaterialAttributes!(
   MaterialAttributes=@n8
   BaseColor=@n81
   Metallic=Multiply(A=Constant() {r=0.95}, B=Saturate(None=@n73))
   Specular=Constant() {r=0.5}
   Roughness=Constant() {r=0.4}
  ) {attributeSetTypes [
    #1
    #2
    #3
    #4
   ]}
  n83 = Multiply(
   A=Max(A=@n22, B=Max(A=@n72.G, B=@n72.R))
   B=FunctionInput(Preview=Constant() {r=1}) {bUsePreviewValueAsDefault=true, inputName=CMask, inputType=FunctionInput_Scalar, sortPriority=4}
  )
  FunctionOutput(None=StaticSwitchParameter(True=BlendMaterialAttributes(A=@n8, B=@n82, Alpha=@n83) {vertexAttributeBlendType=UseA}, False=@n8) {group=Customization_Wrap, parameterName=UseBannerWrap})
  n84 = ScalarParameter() {group=Banner_Weapons, parameterName=BannerPosition_X_Right}
  n85 = ScalarParameter() {group=Banner_Weapons, parameterName=BannerPosition_Y_Right}
  n86 = ScalarParameter() {group=Banner_Vehicle, parameterName=BannerPosition_X_Right_Vehicle}
  n87 = ScalarParameter() {group=Banner_Vehicle, parameterName=BannerPosition_Y_Right_Vehicle}
  LinearInterpolate(
   A=LinearInterpolate(A=AppendVector(A=@n84, B=@n85), B=AppendVector(A=@n86, B=@n87), Alpha=@n12)
   B=AppendVector(
    A=ScalarParameter() {group=Banner_Mech, parameterName=BannerPosition_X_Right_Mech}
    B=ScalarParameter() {group=Banner_Mech, parameterName=BannerPosition_Y_Right_Mech}
   )
   Alpha=@n15
  )
  LinearInterpolate(
   A=ScalarParameter() {group=Banner_Vehicle, parameterName=RotateBanner_Right_Vehicle}
   B=ScalarParameter() {group=Banner_Mech, parameterName=RotateBanner_Right_Mech}
   Alpha=@n15
  )
  n88 = ScalarParameter() {defaultValue=0.9, group=Banner_Mech, parameterName=BannerScale2_Mech}
  LinearInterpolate(A=@n34, B=@n88, Alpha=@n15)
  n89 = ScalarParameter() {group=Banner_Weapons, parameterName=BannerPosition_X_Top}
  n90 = ScalarParameter() {group=Banner_Weapons, parameterName=BannerPosition_Y_Top}
  n91 = ScalarParameter() {group=Banner_Vehicle, parameterName=BannerPosition_X_Top_Vehicle}
  n92 = ScalarParameter() {group=Banner_Vehicle, parameterName=BannerPosition_Y_Top_Vehicle}
  LinearInterpolate(
   A=LinearInterpolate(A=AppendVector(A=@n89, B=@n90), B=AppendVector(A=@n91, B=@n92), Alpha=@n12)
   B=AppendVector(
    A=ScalarParameter() {group=Banner_Mech, parameterName=BannerPosition_X_Top_Mech}
    B=ScalarParameter() {group=Banner_Mech, parameterName=BannerPosition_Y_Top_Mech}
   )
   Alpha=@n15
  )
  LinearInterpolate(A=@n34, B=@n88, Alpha=@n15)
  LinearInterpolate(
   A=ScalarParameter() {group=Banner_Vehicle, parameterName=RotateBanner_Left_Vehicle}
   B=ScalarParameter() {group=Banner_Mech, parameterName=RotateBanner_Left_Mech}
   Alpha=@n15
  )
  n93 = ScalarParameter() {group=Banner_Weapons, parameterName=BannerPosition_X_Left}
  n94 = ScalarParameter() {group=Banner_Weapons, parameterName=BannerPosition_Y_Left}
  n95 = ScalarParameter() {group=Banner_Vehicle, parameterName=BannerPosition_X_Left_Vehicle}
  n96 = ScalarParameter() {group=Banner_Vehicle, parameterName=BannerPosition_Y_Left_Vehicle}
  LinearInterpolate(
   A=LinearInterpolate(A=AppendVector(A=@n93, B=@n94), B=AppendVector(A=@n95, B=@n96), Alpha=@n12)
   B=AppendVector(
    A=ScalarParameter() {group=Banner_Mech, parameterName=BannerPosition_X_Left_Mech}
    B=ScalarParameter() {group=Banner_Mech, parameterName=BannerPosition_Y_Left_Mech}
   )
   Alpha=@n15
  )
  LinearInterpolate(A=@n34, B=@n88, Alpha=@n15)
  MaterialFunctionCall(UVs=@n7, Texture Scale=Constant() {r=15}) {materialFunction=&ScaleUVsByCenter}
  n97 = ScalarParameter() {defaultValue=0.5, parameterName=BW_OffsetMidpoint-Green}
  n98 = PreSkinnedLocalBounds()
  n99 = Clamp(None=@n98.Min, Min=Constant()) {clampMode=CMODE_ClampMin}
  n100 = ScalarParameter() {defaultValue=1.5, parameterName=f3478898kui7}
  n101 = Add(A=Multiply(A=Constant() {r=-0.5}, B=Multiply(A=@n97, B=@n98.Extents)), B=Add(A=Abs(None=@n98.Min), B=Multiply(A=@n99, B=@n100)))
  MaterialFunctionCall(
   X=ComponentMask(None=Add(A=@n101, B=@n19)) {g=true}
   Y=Add(A=ComponentMask(None=@n19) {b=true}, B=ScalarParameter() {defaultValue=-8, parameterName=TryAddingFlatNumberToLocalPosi})
  ) {materialFunction=&MakeFloat2}
  n102 = Constant() {r=0.5}
  n103 = Constant() {r=0.25}
  n104 = MaterialFunctionCall(Float2=@n49.UVs) {materialFunction=&BreakOutFloat2Components}
  n105 = Constant() {r=0.9}
  Multiply(
   A=SphereMask(A=@n102, B=Multiply(A=@n103, B=@n104.R) {constA=0.25, constB=2}, Radius=@n102, Hardness=@n105) {attenuationRadius=0.5}
   B=SphereMask(A=@n104.G, B=@n102, Radius=@n102, Hardness=@n105) {attenuationRadius=0.5}
  )
  NamedRerouteDeclaration(None=FunctionInput() {bUsePreviewValueAsDefault=true, inputName=IsCharacter, inputType=FunctionInput_Scalar}) {name=IsCharacter, nodeColor=<rgba 0 0.9253 1 1>}
  n106 = FunctionInput() {bUsePreviewValueAsDefault=true, inputName=IsStatsB, inputType=FunctionInput_Vector2}
  NamedRerouteDeclaration(None=ComponentMask(None=@n106) {r=true}) {name=isBackpack, nodeColor=<rgba 0 0.9253 1 1>}
  NamedRerouteDeclaration(None=ComponentMask(None=@n106) {g=true}) {name=IsPickaxe, nodeColor=<rgba 0 0.9253 1 1>}
  Reroute(None=@n22)
  StaticSwitchParameter(True=@n71, False=@n70) {group=Banner, parameterName=Banner_Weapon_Icon}
  StaticSwitchParameter(True=@n81, False=Constant()) {parameterName=banner make it work}
  TextureObject() {texture=&T-Banners-Icons-Chevrons-L}
  TextureObject() {texture=/Game/2dAssets/Banners/Season8/T-Banner-Icons-S8-CompletionPirate-L}
  TextureObject() {texture=/Game/UI/Foundation/Textures/Banner/Shapes/banner_shape_standard_Mipmaps}
 ]}
