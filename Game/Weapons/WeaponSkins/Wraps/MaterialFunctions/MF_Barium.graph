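# Material function for the Barium weapon wrap, layering base and accent materials through input masks when the Wrap_Barium static switch is on.
# Side-projected mask textures, placed by per-item-type skull parameters, add to MF_Barium_Streams results.
# Three UV-offset copies, an aberration-style split, blend four colors, roughness and specular values.
# Time-stamp parameters create four normalized reactive timers, alongside a sine-driven UV shake.
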MaterialFunction /Game/Weapons/WeaponSkins/Wraps/MaterialFunctions/MF_Barium
graph {nodes [
  Constant2Vector() {g=0.005, r=0.005}
  Constant2Vector() {r=-0.005}
  Constant() {r=0.25}
  n1 = TextureSample() {samplerType=SAMPLERTYPE_Masks, texture=/Game/Weapons/FORT_Rifles/Textures/T_Scar_CM}
  n2 = FunctionInput(Preview=@n1.R) {bUsePreviewValueAsDefault=true, inputName=Mask Primary, inputType=FunctionInput_Scalar, sortPriority=24, previewValue=<1 0 0 1>}
  n3 = FunctionInput(Preview=@n1.B) {bUsePreviewValueAsDefault=true, inputName=Mask Secondary, inputType=FunctionInput_Scalar, sortPriority=24, previewValue=<1 0 0 1>}
  n4 = BlendMaterialAttributes(A=NamedRerouteUsage(), B=NamedRerouteUsage(), Alpha=Saturate(None=Add(A=@n2, B=@n3))) {vertexAttributeBlendType=UseA}
  n5 = NamedRerouteUsage()
  n6 = NamedRerouteDeclaration(None=@n3) {name=Msk_Accent, nodeColor=<rgba 0 1 0.7656 1>}
  n7 = TextureSample() {samplerType=SAMPLERTYPE_Normal, texture=/Game/Weapons/FORT_Rifles/T_SCAR_N}
  n8 = Constant3Vector() {constant=<rgba 1 0 0 0>}
  n9 = GetMaterialAttributes(None=FunctionInput(Preview=SetMaterialAttributes!(MaterialAttributes, Normal=@n7.RGB, BaseColor=@n8) {attributeSetTypes [#1, #2]}) {inputName=Material, inputType=FunctionInput_MaterialAttributes}) {attributeGetTypes [
    #2
    #3
    #1
    #4
   ]}
  FunctionOutput(None=StaticSwitchParameter(
    True=BlendMaterialAttributes(A=@n4, B=@n5, Alpha=@n6) {vertexAttributeBlendType=UseA}
    False=NamedRerouteDeclaration(None=@n9.MaterialAttributes) {name=Root_Material, nodeColor=<rgba 0.7812 1 0 1>}
   ) {group=Customization_Wrap, parameterName=Wrap_Barium})
  n10 = ComponentMask(None=VectorParameter() {group=Wrap_Barium, parameterName=Barium_aberration_Amount, defaultValue=<rgba 0.0075 0.0075 0 1>}.RGB) {g=true, r=true}
  Multiply(A=@n10, B=Constant2Vector() {g=-1, r=-1})
  Multiply(A=MaterialFunctionCall(Value_Morning=Constant() {r=0.5}, Value_Day=Constant() {r=1}, Value_Evening=Constant() {r=0.5}, Value_Night=Constant() {r=0.25}) {materialFunction=&Scalar_Time_of_Day}.Result)
  n11 = MaterialFunctionCall(Float4=NamedRerouteUsage()) {materialFunction=&BreakOutFloat4Components}
  n12 = MaterialFunctionCall!(
   For_Safe_Zones
   Normal_Map
   Caculated_Bounds=NamedRerouteUsage()
   Pre_Skinned_Normals=NamedRerouteUsage()
   Is Vehicle=@n11.R
   Is VehicleSide=@n11.R
   Is Mech=@n11.R
   Vehicle_Rotation
   UV_Rotation Bypass
   Mech_Rotation
   Uvs_Offset
   Offset-after rotation
   Uvs_TileRate
  ) {materialFunction=&/MF_Projection_Side_Normal}
  n13 = Add(A=Add(A=@n12.UVs No Offsets/Tiling, B=Constant2Vector() {g=0.5, r=-0.5}), B=ComponentMask(None=NamedRerouteUsage()) {g=true, r=true})
  n14 = Abs(None=Subtract(
    A=MaterialFunctionCall(Vector 2=@n13) {materialFunction=/Engine/Functions/Engine_MaterialFunctions02/Utility/VectorLength}.V2 Length
    B=Constant() {r=0.4}
   ) {constB=0})
  n15 = Multiply(A=OneMinus(None=SmoothStep(Value=@n14) {constMax=0.05})) {constB=0.02}
  Multiply(A=TextureSample(UVs=Add(A=@n12.UVs No Offsets/Tiling, B=@n15)) {texture=/Game/Weapons/WeaponSkins/Wraps/Textures/UVCheaker}.RGB)
  n16 = VectorParameter() {group=Wrap_Barium_Accent, parameterName=Barium_Accent_BaseColor, defaultValue=<rgba 0.0194 0.0212 0.0194 1>}
  n17 = SetMaterialAttributes!(
   MaterialAttributes=NamedRerouteUsage()
   BaseColor=@n16.RGB
   Roughness=ScalarParameter() {defaultValue=0.3, group=Wrap_Barium_Accent, parameterName=Barium_Accent_Roughness}
   Specular=ScalarParameter() {defaultValue=0.5, group=Wrap_Barium_Accent, parameterName=Barium_Accent_Spec}
   Metallic=ScalarParameter() {group=Wrap_Barium_Accent, parameterName=Barium_Accent_Metalic}
   Normal
   EmissiveColor
  ) {attributeSetTypes [
    #2
    #3
    #5
    #4
    #1
    #6
   ]}
  NamedRerouteDeclaration(None=@n17) {name=Mat_Accent, nodeColor=<rgba 0 0.3906 1 1>}
  n18 = FunctionInput(Preview=@n1.G) {bUsePreviewValueAsDefault=true, inputName=Mask Crevise, inputType=FunctionInput_Scalar, sortPriority=24, previewValue=<1 0 0 1>}
  NamedRerouteDeclaration(None=@n18) {name=Msk_Curvature, nodeColor=<rgba 0 1 0.7422 1>}
  NamedRerouteDeclaration(None=TextureObjectParameter() {group=Wrap_Barium, parameterName=Barium__Tex_Warp, texture=&T_BlacksRGB}) {name=Tex_Warp, nodeColor=<rgba 0 1 0.8359 1>}
  n19 = Subtract(A=Time(), B=ScalarParameter() {defaultValue=-9999999, group=Wrap_Barium_Reactive, parameterName=Barium_Time_BPA})
  n20 = ScalarParameter() {defaultValue=0.25, group=Wrap_Barium_Reactive, parameterName=Barium_Duration}
  NamedRerouteDeclaration(None=Saturate(None=Divide(A=@n19, B=@n20))) {name=reactiveA, nodeColor=<rgba 0 1 0.9531 1>}
  n21 = Subtract(A=Time(), B=ScalarParameter() {defaultValue=-9999999, group=Wrap_Barium_Reactive, parameterName=Barium_Time_BPB})
  NamedRerouteDeclaration(None=Saturate(None=Divide(A=@n21, B=@n20))) {name=reactiveB, nodeColor=<rgba 0 1 0.9531 1>}
  n22 = Subtract(A=Time(), B=ScalarParameter() {defaultValue=-9999999, group=Wrap_Barium_Reactive, parameterName=Barium_Time_BPC})
  NamedRerouteDeclaration(None=Saturate(None=Divide(A=@n22, B=@n20))) {name=reactiveC, nodeColor=<rgba 0 1 0.9531 1>}
  n23 = Subtract(A=Time(), B=ScalarParameter() {defaultValue=-9999999, group=Wrap_Barium_Reactive, parameterName=Barium_Time_BPD})
  n24 = ScalarParameter() {defaultValue=0.25, group=Wrap_Barium_Reactive, parameterName=Barium_Shake_Duration}
  NamedRerouteDeclaration(None=Saturate(None=Divide(A=@n23, B=@n24))) {name=reactiveD, nodeColor=<rgba 0 1 0.9531 1>}
  n25 = ScalarParameter() {group=Wrap_Barium_UVs, parameterName=Barium_Skull_Rotation_Vehicle, sortPriority=150}
  n26 = ScalarParameter() {group=Wrap_Barium_UVs, parameterName=Barium_Skull_Rotation_Vehicle_Side, sortPriority=150}
  n27 = MaterialFunctionCall(Float4=NamedRerouteUsage()) {materialFunction=&BreakOutFloat4Components}
  n28 = LinearInterpolate(
   A=LinearInterpolate(
    A=ScalarParameter() {group=Wrap_Barium_UVs, parameterName=Barium_Skull_Rotation_Weapon, sortPriority=150}
    B=LinearInterpolate(A=@n25, B=@n26, Alpha=@n27.G)
    Alpha=@n27.R
   )
   B=LinearInterpolate(
    A=ScalarParameter() {group=Wrap_Barium_UVs, parameterName=Barium_Skull_Rotation_Mech, sortPriority=150}
    B=ScalarParameter() {group=Wrap_Barium_UVs, parameterName=Barium_Skull_Rotation_Character, sortPriority=150}
    Alpha=@n27.A
   )
   Alpha=@n27.B
  )
  n29 = ScalarParameter() {group=Wrap_Barium_UVs, parameterName=Barium_Skull_Rotation_Backpack, sortPriority=150}
  n30 = MaterialFunctionCall(Float2=NamedRerouteUsage()) {materialFunction=&BreakOutFloat2Components}
  NamedRerouteDeclaration(None=LinearInterpolate(
    A=LinearInterpolate(A=@n28, B=@n29, Alpha=@n30.R)
    B=ScalarParameter() {group=Wrap_Barium_UVs, parameterName=Barium_Skull_Rotation_Pickaxe, sortPriority=150}
    Alpha=@n30.G
   )) {name=UVs_Skull_Rotation, nodeColor=<rgba 0 0.7188 1 1>}
  n31 = VectorParameter() {group=Wrap_Barium, parameterName=Barium_Color_D, defaultValue=<rgba 0.0194 0.0212 0.0194 1>}
  n32 = NamedRerouteUsage()
  n33 = VectorParameter() {group=Wrap_Barium_Reactive, parameterName=Barium_Shake_Amount, defaultValue=<rgba 0.25 0 0 1>}
  n34 = MaterialFunctionCall(Float4=NamedRerouteUsage()) {materialFunction=&BreakOutFloat4Components}
  n35 = NamedRerouteUsage()
  n36 = Add(
   A=Multiply(A=Sine(None=@n32), B=ComponentMask(None=@n33.RGB) {g=true, r=true}) {constB=0.15}
   B=MaterialFunctionCall(
    For_Safe_Zones=StaticBool() {value=true}
    Caculated_Bounds=NamedRerouteUsage()
    Pre_Skinned_Normals=NamedRerouteUsage()
    Is Vehicle=@n34.R
    Is VehicleSide=@n34.R
    Is Mech=@n34.R
    UV_Rotation Bypass=NamedRerouteUsage()
    Uvs_Offset=ComponentMask(None=@n35) {b=true, g=true, r=true}
    Uvs_TileRate=ComponentMask(None=@n35) {a=true}
   ) {materialFunction=&MF_Projection_Side_Normal}.UVs
  )
  n37 = NamedRerouteUsage()
  n38 = Constant2Vector() {g=0.0025, r=0.0025}
  n39 = NamedRerouteUsage()
  n40 = VectorParameter() {group=Wrap_Barium_Streams, parameterName=Barium_Streams_Speed, defaultValue=<rgba 0.0025 -0.5 0 0>}
  n41 = VectorParameter() {group=Wrap_Barium_Streams, parameterName=Barium_StreamsA_TileRate, defaultValue=<rgba 4 0.2 0 1>}
  n42 = VectorParameter() {group=Wrap_Barium_Streams, parameterName=Barium_StreamsB_TileRate, defaultValue=<rgba 3 0.2 0 1>}
  n43 = NamedRerouteUsage()
  n44 = NamedRerouteUsage()
  n45 = NamedRerouteUsage()
  n46 = VectorParameter() {group=Wrap_Barium_Streams, parameterName=Barium_StreamsWarp_TileRate_A, defaultValue=<rgba 1 1.5 0 1>}
  n47 = VectorParameter() {group=Wrap_Barium_Streams, parameterName=Barium_StreamsWarp_Speed, defaultValue=<rgba -0.1 -0.1 0 1>}
  n48 = Add(
   A=TextureSample(UVs=Add(A=@n36, B=@n10) {constB=0.01}, Tex=@n37).R
   B=MaterialFunctionCall!(
    Uvs=Add(A=@n12.UVs No Offsets/Tiling, B=@n38)
    isstates=@n39
    Texture_Msk=NamedRerouteUsage()
    Texture_Warp=NamedRerouteUsage()
    Location/Scale=NamedRerouteUsage()
    Streams_Speed=@n40.RGB
    StreamsA_TileRate=@n41.RGB
    StreamsB_TileRate=@n42.RGB
    Warp_PositionA=@n43
    Warp_PositionC=@n44
    Warp_PositionB=@n45
    StreamsWarp_TileRate=@n46.RGB
    StreamsWarp_Speed=@n47.RGB
   ) {materialFunction=&/MF_Barium_Streams}.Result
  )
  n49 = Constant2Vector() {g=0.0025, r=0.0025}
  n50 = Add(
   A=TextureSample(UVs=Add(A=@n36, B=@n10), Tex=@n37).R
   B=MaterialFunctionCall(
    Uvs=Add(A=@n12.UVs No Offsets/Tiling, B=@n49)
    isstates=@n39
    Texture_Msk=NamedRerouteUsage()
    Texture_Warp=NamedRerouteUsage()
    Location/Scale=NamedRerouteUsage()
    Streams_Speed=@n40.RGB
    StreamsA_TileRate=@n41.RGB
    StreamsB_TileRate=@n42.RGB
    Warp_PositionA=@n43
    Warp_PositionC=@n44
    Warp_PositionB=@n45
    StreamsWarp_TileRate=@n46.RGB
    StreamsWarp_Speed=@n47.RGB
   ) {materialFunction=&MF_Barium_Streams}.Result
  )
  n51 = Add(
   A=TextureSample(UVs=@n36, Tex=@n37).R
   B=MaterialFunctionCall(
    Uvs=@n12.UVs No Offsets/Tiling
    isstates=@n39
    Texture_Msk=NamedRerouteUsage()
    Texture_Warp=NamedRerouteUsage()
    Location/Scale=NamedRerouteUsage()
    Streams_Speed=@n40.RGB
    StreamsA_TileRate=@n41.RGB
    StreamsB_TileRate=@n42.RGB
    Warp_PositionA=@n43
    Warp_PositionC=@n44
    Warp_PositionB=@n45
    StreamsWarp_TileRate=@n46.RGB
    StreamsWarp_Speed=@n47.RGB
   ) {materialFunction=&MF_Barium_Streams}.Result
  )
  n52 = LinearInterpolate(
   A=LinearInterpolate(
    A=@n31.RGB
    B=VectorParameter() {group=Wrap_Barium, parameterName=Barium_Color_C, defaultValue=<rgba 0.5271 0.1221 0.0116 1>}.RGB
    Alpha=@n48
   )
   B=LinearInterpolate(
    A=VectorParameter() {group=Wrap_Barium, parameterName=Barium_Color_B, defaultValue=<rgba 0.1651 0.5333 0.0103 1>}.RGB
    B=VectorParameter() {group=Wrap_Barium, parameterName=Barium_Color_A, defaultValue=<rgba 0.5395 0.552 0.013 1>}.RGB
    Alpha=@n50
   )
   Alpha=@n51
  )
  n53 = NamedRerouteUsage()
  n54 = NamedRerouteUsage()
  n55 = ScalarParameter() {defaultValue=0.15, group=Wrap_Barium, parameterName=Barium_Roughness_D}
  n56 = LinearInterpolate(
   A=LinearInterpolate(A=@n55, B=ScalarParameter() {defaultValue=0.3, group=Wrap_Barium, parameterName=Barium_Roughness_C}, Alpha=@n48)
   B=LinearInterpolate(
    A=ScalarParameter() {defaultValue=0.3, group=Wrap_Barium, parameterName=Barium_Roughness_B}
    B=ScalarParameter() {defaultValue=0.19, group=Wrap_Barium, parameterName=Barium_Roughness_A}
    Alpha=@n50
   )
   Alpha=@n51
  )
  n57 = NamedRerouteUsage()
  n58 = NamedRerouteUsage()
  n59 = ScalarParameter() {defaultValue=0.85, group=Wrap_Barium, parameterName=Barium_Spec_D}
  n60 = LinearInterpolate(
   A=LinearInterpolate(A=@n59, B=ScalarParameter() {defaultValue=0.7, group=Wrap_Barium, parameterName=Barium_Spec_C}, Alpha=@n48)
   B=LinearInterpolate(
    A=ScalarParameter() {defaultValue=0.7, group=Wrap_Barium, parameterName=Barium_Spec_B}
    B=ScalarParameter() {defaultValue=0.7, group=Wrap_Barium, parameterName=Barium_Spec_A}
    Alpha=@n50
   )
   Alpha=@n51
  )
  n61 = NamedRerouteUsage()
  n62 = NamedRerouteUsage()
  n63 = SetMaterialAttributes(
   MaterialAttributes=NamedRerouteUsage()
   BaseColor=LinearInterpolate(
    A=@n52
    B=LinearInterpolate(A=@n31.RGB, B=@n52, Alpha=@n53)
    Alpha=MaterialFunctionCall(Float4=@n54) {materialFunction=&BreakOutFloat4Components}.R
   )
   Roughness=LinearInterpolate(
    A=@n56
    B=LinearInterpolate(A=@n55, B=@n56, Alpha=@n57)
    Alpha=MaterialFunctionCall(Float4=@n58) {materialFunction=&BreakOutFloat4Components}.R
   )
   Specular=LinearInterpolate(
    A=@n60
    B=LinearInterpolate(A=@n59, B=@n60, Alpha=@n61)
    Alpha=MaterialFunctionCall(Float4=@n62) {materialFunction=&BreakOutFloat4Components}.R
   )
   Metallic=ScalarParameter() {defaultValue=0.5, group=Wrap_Barium, parameterName=Barium_Metalic}
  ) {attributeSetTypes [
    #2
    #3
    #5
    #4
    #1
    #6
   ]}
  NamedRerouteDeclaration(None=@n63) {name=Mat_Base, nodeColor=<rgba 0 0.3906 1 1>}
  n64 = VectorParameter() {group=Wrap_Barium_UVs, parameterName=Barium_Skull_Offset_Scale_Weapon, defaultValue=<rgba -0.4 0 0 0.18>}
  n65 = VectorParameter() {group=Wrap_Barium_UVs, parameterName=Barium_Skull_Offset_Scale_Vehicle, defaultValue=<rgba 0 0 0 1>}
  n66 = AppendVector(A=@n65.RGB, B=@n65.RGB)
  n67 = VectorParameter() {group=Wrap_Barium_UVs, parameterName=Barium_Skull_Offset_Scale_VehicleSide, defaultValue=<rgba 0 0 0 1>}
  n68 = AppendVector(A=@n67.RGB, B=@n67.RGB)
  n69 = MaterialFunctionCall(Float4=NamedRerouteUsage()) {materialFunction=&BreakOutFloat4Components}
  n70 = VectorParameter() {group=Wrap_Barium_UVs, parameterName=Barium_Skull_Offset_Scale_Mech, defaultValue=<rgba 0 0 0 1>}
  n71 = VectorParameter() {group=Wrap_Barium_UVs, parameterName=Barium_Skull_Offset_Offset_Scale_Character, defaultValue=<rgba 0 0 0 1>}
  n72 = LinearInterpolate(
   A=LinearInterpolate(A=AppendVector(A=@n64.RGB, B=@n64.RGB), B=LinearInterpolate(A=@n66, B=@n68, Alpha=@n69.G), Alpha=@n69.R)
   B=LinearInterpolate(A=AppendVector(A=@n70.RGB, B=@n70.RGB), B=AppendVector(A=@n71.RGB, B=@n71.RGB), Alpha=@n69.A)
   Alpha=@n69.B
  )
  n73 = VectorParameter() {group=Wrap_Barium_UVs, parameterName=Barium_Skull_Offset_Offset_Scale_Backpack, defaultValue=<rgba 0 0 0 1>}
  n74 = AppendVector(A=@n73.RGB, B=@n73.RGB)
  n75 = MaterialFunctionCall(Float2=NamedRerouteUsage()) {materialFunction=&BreakOutFloat2Components}
  n76 = VectorParameter() {group=Wrap_Barium_UVs, parameterName=Barium_Skull_Offset_Offset_Scale_Pickaxe, defaultValue=<rgba 0 0 0 1>}
  NamedRerouteDeclaration(None=LinearInterpolate(A=LinearInterpolate(A=@n72, B=@n74, Alpha=@n75.R), B=AppendVector(A=@n76.RGB, B=@n76.RGB), Alpha=@n75.G)) {name=UVs_Skull_Offset_Scale, nodeColor=<rgba 0 0.625 1 1>}
  NamedRerouteDeclaration(None=TextureObjectParameter() {group=Wrap_Barium, parameterName=Barium__Tex_Msk2, texture=&T_BlacksRGB}) {name=Tex_Msk2, nodeColor=<rgba 0 1 0.8359 1>}
  n77 = MaterialFunctionCall(Color=Abs(None=NamedRerouteUsage())) {materialFunction=&SplitComponents}
  n78 = MaterialFunctionCall(Float4=NamedRerouteUsage()) {materialFunction=&BreakOutFloat4Components}
  n79 = LinearInterpolate(A=@n77.R, B=LinearInterpolate(A=@n77.B, B=@n77.B, Alpha=@n78.G), Alpha=@n78.R)
  NamedRerouteDeclaration(None=MaterialFunctionCall(
    In=LinearInterpolate(A=@n79, B=@n77.G, Alpha=@n78.B)
    Contrast=ScalarParameter() {defaultValue=1, group=Wrap_Barium, parameterName=Barium_Blend_Contrast}
   ) {materialFunction=&CheapContrast}.Result) {name=Tri_Mask, nodeColor=<rgba 1 0.4688 0 1>}
  NamedRerouteDeclaration(None=FunctionInput(Preview=Constant4Vector().RGBA) {bUsePreviewValueAsDefault=true, inputName=IsState, inputType=FunctionInput_Vector4, sortPriority=4}) {name=IsState, nodeColor=<rgba 0.1497 1 0.1633 1>}
  NamedRerouteDeclaration(None=FunctionInput(Preview=Constant2Vector()) {bUsePreviewValueAsDefault=true, inputName=IsState_B, inputType=FunctionInput_Vector2, sortPriority=4}) {name=IsState_B, nodeColor=<rgba 0.1497 1 0.1633 1>}
  n80 = PreSkinnedLocalBounds()
  n81 = Subtract(A=VertexInterpolator(VS=PreSkinnedPosition()).PS, B=@n80.Min)
  n82 = MaterialFunctionCall(Color=Subtract(A=@n80.Max, B=@n80.Max)) {materialFunction=&SplitComponents}
  n83 = Max(A=Max(A=@n82.R, B=@n82.R), B=@n82.B)
  NamedRerouteDeclaration(None=FunctionInput(Preview=Divide(A=@n81, B=@n83)) {bUsePreviewValueAsDefault=true, inputName=Calcated_Bounds, sortPriority=11}) {name=Caclcated_Bounds, nodeColor=<rgba 0 0.9297 1 1>}
  NamedRerouteDeclaration(None=TextureObjectParameter() {group=Wrap_Barium, parameterName=Barium__Tex_Msk, texture=&T_BlacksRGB}) {name=Tex_Msk1, nodeColor=<rgba 0 1 0.8359 1>}
  n84 = PreSkinnedNormal()
  NamedRerouteDeclaration(None=FunctionInput(Preview=VertexInterpolator(VS=@n84).PS) {bUsePreviewValueAsDefault=true, inputName=Pre-Skinned Local Normal, sortPriority=10}) {name=PreSkin_Normal, nodeColor=<rgba 0.8281 1 0 1>}
  NamedRerouteDeclaration(None=@n9.Normal) {name=Root_Normal, nodeColor=<rgba 0.4062 0 1 1>}
 ]}
