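# Material function for the Beats wrap skin on weapons, vehicles and mechs, applied through customization masks when the Wrap_Beats static switch is on.
# Three panning top/side projections are height-lerped against a shared gradient mask into hue-shifted emissive strands that can darken the base color.
# Glow brightens on fire via GameTimeFromBP, fades over a set duration, and scales with time of day.
# Warp-UV and reactive-opacity nodes are left unconnected.
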MaterialFunction /Game/Weapons/WeaponSkins/Wraps/MaterialFunctions/MF_Beats_Wrap
graph {nodes [
  n1 = FunctionInput() {inputName=Material, inputType=FunctionInput_MaterialAttributes}
  n2 = GetMaterialAttributes(None=@n1) {attributeGetTypes [
    #1
    #2
    #3
    #4
   ]}
  n3 = FunctionInput(Preview=Constant3Vector() {constant=<rgba 1 0 0 0>}) {bUsePreviewValueAsDefault=true, inputName=ForwardAxis}
  n4 = VectorParameter() {group=BEATS Lerp, parameterName=BEATS_Mask_WeaponOffset_ CE, sortPriority=80, defaultValue=<rgba 0 -1 0 1>}
  n5 = ScalarParameter() {defaultValue=0.5, group=BEATS Lerp, parameterName=BEATS_Mask_WeaponFalloff CE, sortPriority=80}
  n6 = VectorParameter() {group=BEATS Lerp, parameterName=BEATS_Mask_VehicleOffset_ CE, sortPriority=81, defaultValue=<rgba -1 0 0 1>}
  n7 = ScalarParameter() {defaultValue=0.5, group=BEATS Lerp, parameterName=BEATS_Mask_VehicleFalloff CE, sortPriority=81}
  n8 = VectorParameter() {group=BEATS Lerp, parameterName=BEATS_Mask_MechOffset_ CE, sortPriority=82, defaultValue=<rgba 0 0 -1 1>}
  n9 = ScalarParameter() {group=BEATS Lerp, parameterName=BEATS_Mask_MechFalloff CE, sortPriority=82}
  n10 = VectorParameter() {group=BEATS Lerp, parameterName=BEATS_Mask_UV_Spread_XY CE, sortPriority=83, defaultValue=<rgba 1 1 1 0>}
  n11 = VectorParameter() {group=BEATS Lerp, parameterName=BEATS_Mask_UV_Spread_Mech_XY CE, sortPriority=84, defaultValue=<rgba 1 1 1 0>}
  n12 = Constant()
  n13 = FunctionInput(Preview=@n12) {bUsePreviewValueAsDefault=true, inputName=isMech, inputType=FunctionInput_Scalar, sortPriority=4}
  n14 = LinearInterpolate(A=@n10.RGB, B=@n11.RGB, Alpha=@n13)
  n15 = FunctionInput(Preview=Constant()) {inputName=isVehicle, inputType=FunctionInput_Scalar, sortPriority=5}
  n16 = LinearInterpolate(
   A=LinearInterpolate(
    A=VectorParameter() {group=BEATS Lerp, parameterName=BEATS_Mask_UV_AnimatedWeapon_XY CE, sortPriority=85}.RGB
    B=VectorParameter() {group=BEATS Lerp, parameterName=BEATS_Mask_UV_AnimatedVehicle_XY CE, sortPriority=86}.RGB
    Alpha=@n15
   )
   B=VectorParameter() {group=BEATS Lerp, parameterName=BEATS_Mask_UV_AnimatedMech_XY CE, sortPriority=87}.RGB
   Alpha=@n13
  )
  n17 = ScalarParameter() {group=BEATS Lerp, parameterName=BEATS_Mask_WarpAmount, sortPriority=85}
  n18 = FunctionInput(Preview=VertexInterpolator(VS=PreSkinnedPosition()).PS) {bUsePreviewValueAsDefault=true, inputName=Local Position, sortPriority=11}
  n19 = VertexInterpolator(VS=PreSkinnedPosition())
  n20 = PreSkinnedLocalBounds()
  n21 = FunctionInput(Preview=Divide(A=Subtract(A=@n19.PS, B=@n20.Min), B=Subtract(A=@n20.Max, B=@n20.Max))) {bUsePreviewValueAsDefault=true, inputName=Wrap Bounds, sortPriority=12}
  n22 = FunctionInput(Preview=@n12) {bUsePreviewValueAsDefault=true, inputName=Side Project Vehicle, inputType=FunctionInput_Scalar, sortPriority=5}
  n23 = TextureObjectParameter() {group=Beats, parameterName=BEATS_Mask_01, sortPriority=0, texture=&T_BlacksRGB}
  n24 = ChannelMaskParameter(None=TextureSample(
    UVs=MaterialFunctionCall!(
     ForwardAxis=@n3
     Bounds=@n21
     Pre-Skinned Local Position=@n18
     isVehicle=@n15
     isMech=@n13
     Side Project Vehicle=@n22
     WeaponOffset=@n4.RGB
     WeaponFalloff=@n5
     VehicleOffset=@n6.RGB
     VehicleFalloff=@n7
     MechOffset=@n8.RGB
     MechFalloff=@n9
     Invert_Gradent
     UV_Spread_XY=@n14
     UV_Animated_XY=@n16
     Warp_UVs=@n17
     Rotate_Mech
     Rotate_Vehicle_Side
     Rotate_Vehicle
    ) {materialFunction=&/MF_Wrap_Gradent_Creation}.Uvs
    Tex=@n23
   ) {texture=&T_BlacksRGB}.RGBA) {group=BEATS Lerp, parameterName=BEATS_Mask_Gradient_TextureChannel, sortPriority=20}
  n25 = ScalarParameter() {defaultValue=1, group=BEATS Lerp, parameterName=BEATS_Mask_Gradent_Opacity_CE, sortPriority=49}
  n26 = Multiply(
   A=Clamp(None=@n24.R, Max=ScalarParameter() {defaultValue=1, group=BEATS Lerp, parameterName=BEATS_Mask_Gradent_Clamp Max, sortPriority=51}) {maxDefault=0.99, minDefault=0.01}
   B=@n25
  )
  n27 = LinearInterpolate(
   A=ScalarParameter() {defaultValue=1, group=Beats, parameterName=BEATS_Blend_Contrast, sortPriority=-1}
   B=ScalarParameter() {defaultValue=2, group=Beats, parameterName=BEATS_Blend_Contrast_Mech, sortPriority=-1}
   Alpha=@n13
  )
  n28 = ScalarParameter() {group=Beats, parameterName=BEATS_Fill_alternate_Axis_Opacity 01, sortPriority=2}
  n29 = VectorParameter() {group=Beats, parameterName=BEATS_Fill_alternate_Axis_Color_Mask_01, sortPriority=21, defaultValue=<rgba 0 0 0 1>}
  n30 = AppendVector(A=@n29.RGB, B=@n29.RGB)
  n31 = VectorParameter() {group=BEATS_UVS, parameterName=BEATS_Rotation-Weapon-01---------------------, sortPriority=1, defaultValue=<rgba 0 0 -0.25 1>}
  n32 = VectorParameter() {group=BEATS_UVS, parameterName=BEATS_Rotation_Vehicle_01, sortPriority=2, defaultValue=<rgba 0 0 0 1>}
  n33 = VectorParameter() {
   desc=must activate TS_Vehicle_Side_01 UVs
   group=BEATS_UVS
   parameterName=BEATS_Rotation_Vehicle_Side_01
   sortPriority=2
   defaultValue=<rgba 0 0 0 1>
  }
  n34 = VectorParameter() {group=BEATS_UVS, parameterName=BEATS_Rotation_Mech_01, sortPriority=3, defaultValue=<rgba 0.25 0.25 0 1>}
  n35 = LinearInterpolate(A=LinearInterpolate(A=@n31.RGB, B=LinearInterpolate(A=@n32.RGB, B=@n33.RGB, Alpha=@n22), Alpha=@n15), B=@n34.RGB, Alpha=@n13)
  n36 = FunctionInput(Preview=VertexInterpolator(VS=PreSkinnedNormal()).PS) {bUsePreviewValueAsDefault=true, inputName=Pre-Skinned Local Normal, sortPriority=10}
  n37 = VectorParameter() {group=BEATS_UVS, parameterName=BEATS_Offset_Weapon_01---------------------, sortPriority=10, defaultValue=<rgba 0 0 0 1>}
  n38 = VectorParameter() {group=BEATS_UVS, parameterName=BEATS_Offset_Vehicle_01, sortPriority=11, defaultValue=<rgba 0 0 0 1>}
  n39 = VectorParameter() {
   desc=must activate TS_Vehicle_Side_01 UVs
   group=BEATS_UVS
   parameterName=BEATS_Offset_Vehicle_Side_01
   sortPriority=11
   defaultValue=<rgba 0 0 0 1>
  }
  n40 = LinearInterpolate(A=@n38.RGB, B=@n39.RGB, Alpha=@n22)
  n41 = Add(
   A=@n21
   B=LinearInterpolate(
    A=LinearInterpolate(A=@n37.RGB, B=@n40, Alpha=@n15)
    B=VectorParameter() {group=BEATS_UVS, parameterName=BEATS_Offset_Mech_01, sortPriority=12, defaultValue=<rgba 0 0 0 1>}.RGB
    Alpha=@n13
   )
  )
  n42 = VectorParameter() {group=BEATS_UVS, parameterName=BEATS_Panning_Weapon_01---------------------, sortPriority=40, defaultValue=<rgba 0 0 0 1>}
  n43 = VectorParameter() {group=BEATS_UVS, parameterName=BEATS_Panning_Vehicle_01, sortPriority=41, defaultValue=<rgba 0 0 0 1>}
  n44 = VectorParameter() {
   desc=must activate TS_Vehicle_Side_01 UVs
   group=BEATS_UVS
   parameterName=BEATS_Panning_Vehicle_Side_01
   sortPriority=41
   defaultValue=<rgba 0 0 0 1>
  }
  n45 = LinearInterpolate(A=@n43.RGB, B=@n44.RGB, Alpha=@n22)
  n46 = VectorParameter() {group=BEATS_UVS, parameterName=BEATS_Panning_Mech_01, sortPriority=42, defaultValue=<rgba 0 0 0 1>}
  n47 = Multiply(A=Time(), B=LinearInterpolate(A=LinearInterpolate(A=@n42.RGB, B=@n45, Alpha=@n15), B=@n46.RGB, Alpha=@n13))
  n48 = VectorParameter() {group=BEATS_UVS, parameterName="BEATS_Scale_Weapon_01 A=uniform", sortPriority=30, defaultValue=<rgba 1 1 1 1>}
  n49 = VectorParameter() {group=BEATS_UVS, parameterName="BEATS_Scale_Vehicle_01 A=uniform", sortPriority=31, defaultValue=<rgba 1 1 1 1>}
  n50 = VectorParameter() {
   desc=must activate TS_Vehicle_Side_01 UVs
   group=BEATS_UVS
   parameterName="BEATS_Scale_Vehicle_Side_01 A=uniform"
   sortPriority=31
   defaultValue=<rgba 1 1 1 1>
  }
  n51 = LinearInterpolate(
   A=Multiply(A=@n48.RGB, B=@n48.RGB)
   B=LinearInterpolate(A=Multiply(A=@n49.RGB, B=@n49.RGB), B=Multiply(A=@n50.RGB, B=@n50.RGB), Alpha=@n22)
   Alpha=@n15
  )
  n52 = VectorParameter() {group=BEATS_UVS, parameterName="BEATS_Scale_Mech_01 A=uniform", defaultValue=<rgba 1 1 1 1>}
  n53 = Multiply(A=@n52.RGB, B=@n52.RGB)
  n54 = MaterialFunctionCall!(
   IsCharacter
   IsStatsB
   Texture=@n23
   Rotation_Axis=@n35
   Caculated_Bounds=Multiply(A=Add(A=@n41, B=@n47), B=LinearInterpolate(A=@n51, B=@n53, Alpha=@n13))
   Pre_Skinned_Normals=@n36
   Blend_Contrast=@n27
   Fill_alternate_Axis_Opacity=@n28
   "Fill_alternate_Axis_Color_A=Roughness"=@n30
   Is_Vehicle=@n15
   Is_Mech=@n13
   Enable_Fill_Z_Weapon
   Fill_Z_Weapon_opacity
   Fill_Z_Weapon_Color
  ) {materialFunction=&/MF_TopsSideProjection_Base}
  n55 = Saturate(None=MaterialFunctionCall(
    Transition Phase=Saturate(None=@n26)
    Height Texture=ChannelMaskParameter(None=@n54.Combined) {group=BEATS Lerp, parameterName=TS_Mask_HeightTexture_01, sortPriority=18}.R
    Contrast=ScalarParameter() {defaultValue=1, group=BEATS Lerp, parameterName=BEATS_Mask_HeightLerp_Contrast, sortPriority=45}
   ) {materialFunction=&HeightLerp}.Alpha)
  n56 = ScalarParameter() {defaultValue=1, group=BEATS Lerp, parameterName=BEATS_Mask_Gradent_Opacity_CE_02, sortPriority=149}
  n57 = Multiply(
   A=Clamp(None=@n24.R, Max=ScalarParameter() {defaultValue=1, group=BEATS Lerp, parameterName=BEATS_Mask_Gradent_Clamp Max_02, sortPriority=151}) {maxDefault=0.99, minDefault=0.01}
   B=@n56
  )
  n58 = VectorParameter() {group=BEATS_UVS, parameterName=BEATS_Offset_Weapon_02---------------------, sortPriority=160, defaultValue=<rgba 0 0 0 1>}
  n59 = VectorParameter() {group=BEATS_UVS, parameterName=BEATS_Offset_Vehicle_02, sortPriority=161, defaultValue=<rgba 0 0 0 1>}
  n60 = VectorParameter() {
   desc=must activate TS_Vehicle_Side_02 UVs
   group=BEATS_UVS
   parameterName=BEATS_Offset_Vehicle_Side_02
   sortPriority=161
   defaultValue=<rgba 0 0 0 1>
  }
  n61 = LinearInterpolate(A=@n59.RGB, B=@n60.RGB, Alpha=@n22)
  n62 = VectorParameter() {group=BEATS_UVS, parameterName=BEATS_Offset_Mech_02, sortPriority=162, defaultValue=<rgba 0 0 0 1>}
  n63 = Add(A=@n21, B=LinearInterpolate(A=LinearInterpolate(A=@n58.RGB, B=@n61, Alpha=@n15), B=@n62.RGB, Alpha=@n13))
  n64 = VectorParameter() {group=BEATS_UVS, parameterName=BEATS_Panning_Weapon_02---------------------, sortPriority=180, defaultValue=<rgba 0 0 0 1>}
  n65 = VectorParameter() {group=BEATS_UVS, parameterName=BEATS_Panning_Vehicle_02, sortPriority=181, defaultValue=<rgba 0 0 0 1>}
  n66 = VectorParameter() {
   desc=must activate TS_Vehicle_Side_02 UVs
   group=BEATS_UVS
   parameterName=BEATS_Panning_Vehicle_Side_02
   sortPriority=181
   defaultValue=<rgba 0 0 0 1>
  }
  n67 = LinearInterpolate(A=@n65.RGB, B=@n66.RGB, Alpha=@n22)
  n68 = VectorParameter() {group=BEATS_UVS, parameterName=BEATS_Panning_Mech_02, sortPriority=182, defaultValue=<rgba 0 0 0 1>}
  n69 = Multiply(A=Time(), B=LinearInterpolate(A=LinearInterpolate(A=@n64.RGB, B=@n67, Alpha=@n15), B=@n68.RGB, Alpha=@n13))
  n70 = VectorParameter() {group=BEATS_UVS, parameterName="BEATS_Scale_Weapon_02 A=uniform", sortPriority=170, defaultValue=<rgba 1 1 1 1>}
  n71 = VectorParameter() {group=BEATS_UVS, parameterName="BEATS_Scale_Vehicle_02 A=uniform", sortPriority=171, defaultValue=<rgba 1 1 1 1>}
  n72 = VectorParameter() {
   desc=must activate TS_Vehicle_Side_02 UVs
   group=BEATS_UVS
   parameterName="BEATS_Scale_Vehicle_Side_02 A=uniform"
   sortPriority=171
   defaultValue=<rgba 1 1 1 1>
  }
  n73 = LinearInterpolate(
   A=Multiply(A=@n70.RGB, B=@n70.RGB)
   B=LinearInterpolate(A=Multiply(A=@n71.RGB, B=@n71.RGB), B=Multiply(A=@n72.RGB, B=@n72.RGB), Alpha=@n22)
   Alpha=@n15
  )
  n74 = VectorParameter() {group=BEATS_UVS, parameterName="BEATS_Scale_Mech_02 A=uniform", sortPriority=172, defaultValue=<rgba 1 1 1 1>}
  n75 = Multiply(A=@n74.RGB, B=@n74.RGB)
  n76 = MaterialFunctionCall(
   Texture=@n23
   Rotation_Axis=@n35
   Caculated_Bounds=Multiply(A=Add(A=@n63, B=@n69), B=LinearInterpolate(A=@n73, B=@n75, Alpha=@n13))
   Pre_Skinned_Normals=@n36
   Blend_Contrast=@n27
   Fill_alternate_Axis_Opacity=@n28
   "Fill_alternate_Axis_Color_A=Roughness"=@n30
   Is_Vehicle=@n15
   Is_Mech=@n13
  ) {materialFunction=&MF_TopsSideProjection_Base}
  n77 = Saturate(None=MaterialFunctionCall(
    Transition Phase=Saturate(None=@n57)
    Height Texture=ChannelMaskParameter(None=@n76.Combined) {group=BEATS Lerp, parameterName=TS_Mask_HeightTexture_02, sortPriority=119}.R
    Contrast=ScalarParameter() {defaultValue=1, group=BEATS Lerp, parameterName=BEATS_Mask_HeightLerp_Contrast_02, sortPriority=145}
   ) {materialFunction=&HeightLerp}.Alpha)
  n78 = Multiply(
   A=Clamp(None=@n24.R, Max=ScalarParameter() {defaultValue=1, group=BEATS Lerp, parameterName=BEATS_Mask_Gradent_Clamp Max_03, sortPriority=251}) {maxDefault=0.99, minDefault=0.01}
   B=ScalarParameter() {defaultValue=1, group=BEATS Lerp, parameterName=BEATS_Mask_Gradent_Opacity_CE_03, sortPriority=249}
  )
  n79 = VectorParameter() {group=BEATS_UVS, parameterName=BEATS_Offset_Weapon_03---------------------, sortPriority=260, defaultValue=<rgba 0 0 0 1>}
  n80 = VectorParameter() {group=BEATS_UVS, parameterName=BEATS_Offset_Vehicle_03, sortPriority=261, defaultValue=<rgba 0 0 0 1>}
  n81 = VectorParameter() {
   desc=must activate TS_Vehicle_Side_02 UVs
   group=BEATS_UVS
   parameterName=BEATS_Offset_Vehicle_Side_03
   sortPriority=261
   defaultValue=<rgba 0 0 0 1>
  }
  n82 = LinearInterpolate(A=@n80.RGB, B=@n81.RGB, Alpha=@n22)
  n83 = VectorParameter() {group=BEATS_UVS, parameterName=BEATS_Offset_Mech_03, sortPriority=262, defaultValue=<rgba 0 0 0 1>}
  n84 = Add(A=@n21, B=LinearInterpolate(A=LinearInterpolate(A=@n79.RGB, B=@n82, Alpha=@n15), B=@n83.RGB, Alpha=@n13))
  n85 = VectorParameter() {group=BEATS_UVS, parameterName=BEATS_Panning_Weapon_03---------------------, sortPriority=280, defaultValue=<rgba 0 0 0 1>}
  n86 = VectorParameter() {group=BEATS_UVS, parameterName=BEATS_Panning_Vehicle_03, sortPriority=281, defaultValue=<rgba 0 0 0 1>}
  n87 = VectorParameter() {
   desc=must activate TS_Vehicle_Side_02 UVs
   group=BEATS_UVS
   parameterName=BEATS_Panning_Vehicle_Side_03
   sortPriority=281
   defaultValue=<rgba 0 0 0 1>
  }
  n88 = LinearInterpolate(A=@n86.RGB, B=@n87.RGB, Alpha=@n22)
  n89 = VectorParameter() {group=BEATS_UVS, parameterName=BEATS_Panning_Mech_03, sortPriority=282, defaultValue=<rgba 0 0 0 1>}
  n90 = Multiply(A=Time(), B=LinearInterpolate(A=LinearInterpolate(A=@n85.RGB, B=@n88, Alpha=@n15), B=@n89.RGB, Alpha=@n13))
  n91 = VectorParameter() {group=BEATS_UVS, parameterName="BEATS_Scale_Weapon_03 A=uniform", sortPriority=270, defaultValue=<rgba 1 1 1 1>}
  n92 = VectorParameter() {group=BEATS_UVS, parameterName="BEATS_Scale_Vehicle_03 A=uniform", sortPriority=271, defaultValue=<rgba 1 1 1 1>}
  n93 = VectorParameter() {
   desc=must activate TS_Vehicle_Side_02 UVs
   group=BEATS_UVS
   parameterName="BEATS_Scale_Vehicle_Side_03 A=uniform"
   sortPriority=271
   defaultValue=<rgba 1 1 1 1>
  }
  n94 = LinearInterpolate(
   A=Multiply(A=@n91.RGB, B=@n91.RGB)
   B=LinearInterpolate(A=Multiply(A=@n92.RGB, B=@n92.RGB), B=Multiply(A=@n93.RGB, B=@n93.RGB), Alpha=@n22)
   Alpha=@n15
  )
  n95 = VectorParameter() {group=BEATS_UVS, parameterName="BEATS_Scale_Mech_03 A=uniform", sortPriority=272, defaultValue=<rgba 1 1 1 1>}
  n96 = Multiply(A=@n95.RGB, B=@n95.RGB)
  n97 = MaterialFunctionCall(
   Texture=@n23
   Rotation_Axis=@n35
   Caculated_Bounds=Multiply(A=Add(A=@n84, B=@n90), B=LinearInterpolate(A=@n94, B=@n96, Alpha=@n13))
   Pre_Skinned_Normals=@n36
   Blend_Contrast=@n27
   Fill_alternate_Axis_Opacity=@n28
   "Fill_alternate_Axis_Color_A=Roughness"=@n30
   Is_Vehicle=@n15
   Is_Mech=@n13
  ) {materialFunction=&MF_TopsSideProjection_Base}
  n98 = Saturate(None=MaterialFunctionCall(
    Transition Phase=Saturate(None=@n78)
    Height Texture=ChannelMaskParameter(None=@n97.Combined) {group=BEATS Lerp, parameterName=TS_Mask_HeightTexture_03, sortPriority=219}.R
    Contrast=ScalarParameter() {defaultValue=1, group=BEATS Lerp, parameterName=BEATS_Mask_HeightLerp_Contrast_03, sortPriority=245}
   ) {materialFunction=&HeightLerp}.Alpha)
  n99 = Add(A=Add(A=@n55, B=@n77), B=@n98)
  n100 = Multiply(A=ScalarParameter() {group=BEATS Lerp, parameterName=BEATS_Emissive01-DarkenBase-------, sortPriority=-20}, B=@n99)
  n101 = Multiply(A=ComponentMask(None=@n2.BaseColor) {b=true, desc=Base Texture, g=true, r=true}, B=OneMinus(None=Saturate(None=@n100)))
  n102 = ScalarParameter() {group=Beats, parameterName=BEATS_HueShift_BaseColor 03, sortPriority=219}
  n103 = Multiply(A=Time(), B=ScalarParameter() {group=Beats, parameterName=BEATS_HueShift_BaseColor_Animated 03, sortPriority=220})
  n104 = VectorParameter() {group=BEATS Lerp, parameterName=BEATS_HueShift Color 03, sortPriority=218, defaultValue=<rgba 0.7377 0 0.807 1>}
  n105 = Multiply(
   A=MaterialFunctionCall(Hue Shift Percentage=Add(A=@n102, B=@n103), Texture=@n104.RGB) {materialFunction=&HueShift}.Result
   B=@n98
  )
  n106 = Time()
  n107 = ScalarParameter() {group=00-Reactive-GTFB, parameterName=00-GameTimeFromBP, sortPriority=1}
  n108 = Saturate(None=Divide(
    A=Subtract(A=@n106, B=@n107)
    B=ScalarParameter() {defaultValue=1, group=Beats Reactive, parameterName=BEATS_AnimationDuration, sortPriority=2}
   ))
  n109 = ScalarParameter() {defaultValue=0.5, group=Beats Reactive, parameterName=BEATS_Curve_Amount, sortPriority=3}
  n110 = ScalarParameter() {group=Beats Reactive, parameterName=BEATS_Reactive, sortPriority=-10}
  n111 = LinearInterpolate(A=OneMinus(None=Power(Base=@n108, Exp=@n109)) {desc=inverts curve from 0-1 to 1-0}, B=@n110, Alpha=@n110)
  n112 = LinearInterpolate(
   A=@n105
   B=Multiply(
    A=@n105
    B=ScalarParameter() {defaultValue=5, group=Beats Reactive, parameterName=BEATS_ColorBrightnessTertiaryLerp_Reactive, sortPriority=11}
   )
   Alpha=@n111
  )
  n113 = ScalarParameter() {group=Beats, parameterName=BEATS_HueShift_BaseColor 02, sortPriority=119}
  n114 = Multiply(A=Time(), B=ScalarParameter() {group=Beats, parameterName=BEATS_HueShift_BaseColor_Animated 02, sortPriority=120})
  n115 = VectorParameter() {group=BEATS Lerp, parameterName=BEATS_HueShift Color 02, sortPriority=118, defaultValue=<rgba 0.855 0.0603 0 1>}
  n116 = Multiply(
   A=MaterialFunctionCall(Hue Shift Percentage=Add(A=@n113, B=@n114), Texture=@n115.RGB) {materialFunction=&HueShift}.Result
   B=@n77
  )
  n117 = ScalarParameter() {defaultValue=5, group=Beats Reactive, parameterName=BEATS_ColorBrightness_Reactive, sortPriority=10}
  n118 = LinearInterpolate(A=@n116, B=Multiply(A=@n116, B=@n117), Alpha=@n111)
  n119 = ScalarParameter() {group=Beats, parameterName=BEATS_HueShift_BaseColor 01, sortPriority=119}
  n120 = Multiply(A=Time(), B=ScalarParameter() {group=Beats, parameterName=BEATS_HueShift_BaseColor_Animated 01, sortPriority=120})
  n121 = VectorParameter() {group=BEATS Lerp, parameterName=BEATS_HueShift Color 01, sortPriority=18, defaultValue=<rgba 0.0273 0.8228 0.8388 1>}
  n122 = Multiply(
   A=@n55
   B=MaterialFunctionCall(Hue Shift Percentage=Add(A=@n119, B=@n120), Texture=@n121.RGB) {materialFunction=&HueShift}.Result
  )
  n123 = Multiply(A=@n122, B=@n117)
  n124 = LinearInterpolate(
   A=LinearInterpolate(A=LinearInterpolate(A=@n101, B=@n112, Alpha=@n98), B=LinearInterpolate(A=@n101, B=@n118, Alpha=@n77), Alpha=@n77)
   B=LinearInterpolate(A=@n101, B=LinearInterpolate(A=@n122, B=@n123, Alpha=@n111), Alpha=@n55)
   Alpha=@n55
  )
  Add(A=@n124, B=@n99)
  n125 = ScalarParameter() {defaultValue=1, group=Beats, parameterName=BEATS_WarpUVS-02-TileRate Weapon, sortPriority=203}
  n126 = ScalarParameter() {defaultValue=1, group=Beats, parameterName=BEATS_WarpUVS-02-TileRate Vehicle, sortPriority=204}
  n127 = VectorParameter() {group=Beats, parameterName=BEATS_WarpUVS-02-Anitmated Weapon, sortPriority=207, defaultValue=<rgba 0 0 0 1>}
  n128 = VectorParameter() {group=Beats, parameterName=BEATS_WarpUVS-02-Anitmated Vehicle, sortPriority=208, defaultValue=<rgba 0 0 0 1>}
  n129 = LinearInterpolate(A=@n127.RGB, B=@n128.RGB, Alpha=@n15)
  Add(
   A=Multiply(A=TextureCoordinate(), B=LinearInterpolate(A=@n125, B=@n126, Alpha=@n15))
   B=Multiply(A=Time() {period=1}, B=ComponentMask(None=@n129) {g=true, r=true})
  )
  n130 = TextureSampleParameter2D(UVs=@n54.Combined) {group=Beats, parameterName=BEATS_WarpUVS-02-Texture, sortPriority=202}
  n131 = ScalarParameter() {defaultValue=1, group=Beats, parameterName=BEATS_WarpUVS-02-Amount Weapon, sortPriority=205}
  Add(
   A=@n21
   B=Multiply(A=ChannelMaskParameter(None=@n130.RGBA) {group=Beats, parameterName=BEATS_WarpUVS-02_Channel, sortPriority=201}.R, B=@n131)
  )
  n132 = ScalarParameter() {defaultValue=0.5, group=Beats, parameterName=BEATS_Lerp_Specular, sortPriority=12}
  n133 = ScalarParameter() {defaultValue=6, group=Beats, parameterName=BEATS_SpeculartyFresnel, sortPriority=12}
  n134 = Transform(None=@n2.Normal)
  n135 = Constant() {r=0.5}
  n136 = Constant() {r=1}
  n137 = Constant() {r=0.25}
  n138 = SetMaterialAttributes!(
   MaterialAttributes=@n2.MaterialAttributes
   BaseColor=@n101
   Roughness=LinearInterpolate(A=@n2.Roughness, B=ScalarParameter() {defaultValue=0.5, group=Beats, parameterName=BEATS_Lerp_Roughness, sortPriority=11}, Alpha=@n99)
   Specular=Add(A=LinearInterpolate(A=@n2.Specular, B=@n132, Alpha=@n99), B=Fresnel(ExponentIn=@n133, Normal=@n134))
   EmissiveColor=Multiply(
    A=@n124
    B=MaterialFunctionCall(Value_Morning=@n135, Value_Day=@n136, Value_Evening=@n135, Value_Night=@n137) {materialFunction=&Scalar_Time_of_Day}.Result
   )
   Metallic=Constant()
  ) {attributeSetTypes [
    #2
    #4
    #3
    #5
    #6
   ]}
  n139 = FunctionInput() {bUsePreviewValueAsDefault=true, inputName="Customization Mask 2 (_CM B Top Right)", inputType=FunctionInput_Scalar, sortPriority=25}
  n140 = FunctionInput() {bUsePreviewValueAsDefault=true, inputName="Customization Mask (_CM B Top left)", inputType=FunctionInput_Scalar, sortPriority=24}
  n141 = StaticSwitchParameter(True=BlendMaterialAttributes(A=@n2.MaterialAttributes, B=@n138, Alpha=Add(A=@n139, B=@n140)), False=@n1) {
   desc="Topside Projection (Masked Based) - Emissive 3 Height Lerps strands for parallax - TOD emissive restrictive - Reactive (Glow on Fire)"
   group=Customization_Wrap
   parameterName=Wrap_Beats
  }
  FunctionOutput(None=@n141)
  LinearInterpolate(A=@n56, B=ScalarParameter() {defaultValue=0.8, group=BEATS Lerp, parameterName=BEATS_Reactive_Mask_Gradent_Opacity_CE_02, sortPriority=6})
  LinearInterpolate(
   A=@n25
   B=ScalarParameter() {defaultValue=0.8, group=BEATS Lerp, parameterName=BEATS_Reactive_Mask_Gradent_Opacity_CE, sortPriority=46}
   Alpha=@n111
  )
  LinearInterpolate(A=@n131, B=ScalarParameter() {defaultValue=1, group=Beats, parameterName=BEATS_WarpUVS-02-Amount_Vehicle, sortPriority=206}, Alpha=@n15)
 ]}
