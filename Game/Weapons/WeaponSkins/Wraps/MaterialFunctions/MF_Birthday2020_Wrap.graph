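# Birthday 2020 wrap material function for weapons, vehicles and mechs, passing input attributes through when its static switch is off.
# Projects base color, roughness and normals from pre-skinned bounds, with per-type axes, rotation and scale; mobile skips the normals.
# Customization masks blend that layer and a UV-tiled secondary layer.
# Two reactive confetti layers animate from timestamp parameters, using height-lerped gradients, hue shifts and fades for time-of-day-scaled emissive.
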MaterialFunction /Game/Weapons/WeaponSkins/Wraps/MaterialFunctions/MF_Birthday2020_Wrap
graph {nodes [
  n1 = VectorParameter() {group=Birthday2020_UVs, parameterName="BD_Scale_Weapon_01 A=uniform", sortPriority=30, defaultValue=<rgba 1 1 1 1>}
  n2 = Multiply(A=@n1.RGB, B=@n1.RGB)
  n3 = VectorParameter() {group=Birthday2020_UVs, parameterName="BD_Scale_Vehicle_01 A=uniform", sortPriority=31, defaultValue=<rgba 1 1 1 1>}
  n4 = VectorParameter() {group=Birthday2020_UVs, parameterName="BD_Scale_VehicleSide_01 A=uniform", sortPriority=31, defaultValue=<rgba 1 1 1 1>}
  n5 = FunctionInput(Preview=Constant()) {inputName=isVehicleSide, inputType=FunctionInput_Scalar, sortPriority=5}
  n6 = LinearInterpolate(A=Multiply(A=@n3.RGB, B=@n3.RGB), B=Multiply(A=@n4.RGB, B=@n4.RGB), Alpha=@n5)
  n7 = FunctionInput(Preview=Constant()) {inputName=isVehicle, inputType=FunctionInput_Scalar, sortPriority=5}
  n8 = VectorParameter() {group=Birthday2020_UVs, parameterName="BD_Scale_Mech_01 A=uniform", defaultValue=<rgba 1 1 1 1>}
  n9 = FunctionInput(Preview=Constant()) {bUsePreviewValueAsDefault=true, inputName=isMech, inputType=FunctionInput_Scalar, sortPriority=4}
  n10 = Constant3Vector() {constant=<rgba 1 1 -1 0>}
  n11 = VertexInterpolator(VS=PreSkinnedPosition())
  n12 = PreSkinnedLocalBounds()
  n13 = FunctionInput(Preview=Divide(A=Subtract(A=@n11.PS, B=@n12.Min), B=Subtract(A=@n12.Max, B=@n12.Max))) {bUsePreviewValueAsDefault=true, inputName=Calculated_Bounds, sortPriority=11}
  n14 = VectorParameter() {group=Birthday2020_UVs, parameterName=BD_TextureOffset_Weapon, sortPriority=15, defaultValue=<rgba 0 0 0 1>}
  n15 = VectorParameter() {group=Birthday2020_UVs, parameterName=BD_TextureOffset_Vehicle, sortPriority=15, defaultValue=<rgba 0 0 0 1>}
  n16 = VectorParameter() {group=Birthday2020_UVs, parameterName=BD_TextureOffset_VehicleSide, sortPriority=15, defaultValue=<rgba 0 0 0 1>}
  n17 = LinearInterpolate(A=@n14.RGB, B=LinearInterpolate(A=@n15.RGB, B=@n16.RGB, Alpha=@n5), Alpha=@n7)
  n18 = VectorParameter() {group=Birthday2020_UVs, parameterName=BD_TextureOffset_Mech, sortPriority=15, defaultValue=<rgba 0 0 0 1>}
  n19 = Multiply(
   A=Add(A=Multiply(A=@n13, B=@n10), B=LinearInterpolate(A=@n17, B=@n18.RGB, Alpha=@n9))
   B=LinearInterpolate(A=LinearInterpolate(A=@n2, B=@n6, Alpha=@n7), B=Multiply(A=@n8.RGB, B=@n8.RGB), Alpha=@n9)
  )
  AppendVector(A=OneMinus(None=ComponentMask(None=@n19) {b=true}), B=ComponentMask(None=@n19) {r=true})
  n20 = VectorParameter() {group=Birthday2020_UVs, parameterName=BD_WeaponRotation, sortPriority=-4, defaultValue=<rgba 0 0 -0.25 1>}
  n21 = VectorParameter() {group=Birthday2020_UVs, parameterName=BD_VehicleRotation, sortPriority=-3, defaultValue=<rgba 0 0 0 1>}
  n22 = VectorParameter() {group=Birthday2020_UVs, parameterName=BD_VehicleSideRotation, sortPriority=-3, defaultValue=<rgba 0 0 0 1>}
  n23 = LinearInterpolate(A=@n21.RGB, B=@n22.RGB, Alpha=@n5)
  n24 = VectorParameter() {group=Birthday2020_UVs, parameterName=BD_MechRotation, sortPriority=-2, defaultValue=<rgba 0.2502 0.2502 0 1>}
  n25 = MaterialFunctionCall(Color=LinearInterpolate(A=LinearInterpolate(A=@n20.RGB, B=@n23, Alpha=@n7), B=@n24.RGB, Alpha=@n9)) {materialFunction=&SplitComponents}
  n26 = MaterialFunctionCall(UVs=ComponentMask(None=@n19) {b=true, g=true}, "Rotation Angle (0-1)"=@n25.R) {materialFunction=&CustomRotator}
  n27 = TextureObjectParameter() {
   group=Birthday2020
   parameterName=BD_Normal_01
   samplerType=SAMPLERTYPE_Normal
   sortPriority=13
   texture=&FlatNormalBlue
  }
  n28 = TextureSample(UVs=@n26.Rotated Values, Tex=@n27) {samplerType=SAMPLERTYPE_Normal}
  n29 = FunctionInput(Preview=VertexInterpolator(VS=PreSkinnedNormal()).PS) {bUsePreviewValueAsDefault=true, inputName=Pre-Skinned Local Normal, sortPriority=10}
  n30 = Abs(None=@n29)
  n31 = ComponentMask(None=@n30) {desc=Weapon, r=true}
  n32 = Multiply(A=@n28.R) {constB=-1}
  n33 = AppendVector(A=AppendVector(A=Multiply(A=@n28.B, B=@n31), B=LinearInterpolate(A=@n28.R, B=@n32)), B=Multiply(A=@n28.G) {constB=-1})
  n34 = Add(
   A=RotateAboutAxis(
    NormalizedRotationAxis=Constant3Vector() {constant=<rgba 1 0 0 0>}
    RotationAngle=Multiply(A=@n25.R) {constB=-1}
    PivotPoint=Constant3Vector() {constant=<rgba 0 0 0 1>}
    Position=@n33
   )
   B=@n33
  )
  n35 = FunctionInput() {inputName=Material, inputType=FunctionInput_MaterialAttributes, sortPriority=-1}
  n36 = GetMaterialAttributes(None=@n35) {attributeGetTypes [#1, #2, #3]}
  n37 = ScalarParameter() {defaultValue=6, group=Birthday2020, parameterName=BD_MaskedNormalScale_Weapon, sortPriority=52}
  n38 = ScalarParameter() {defaultValue=4, group=Birthday2020, parameterName=BD_MaskedNormalScale_Vehicle, sortPriority=53}
  n39 = TextureSampleParameter2D(UVs=Multiply(A=TextureCoordinate(), B=LinearInterpolate(A=@n37, B=@n38, Alpha=@n7))) {
   group=Birthday2020
   parameterName=BD_MaskedNormal_01
   samplerType=SAMPLERTYPE_Normal
   sortPriority=50
   texture=/Game/Weapons/WeaponSkins/Wraps/Textures/Birthday2020/T_Wrap_Birthday2020_N
  }
  n40 = StaticSwitchParameter(
   True=Constant3Vector() {constant=<rgba 0 0 1 0>}
   False=LinearInterpolate(
    A=@n36.Normal
    B=MaterialFunctionCall(BaseNormal=@n36.Normal, AdditionalNormal=@n39.RGB) {materialFunction=&BlendAngleCorrectedNormals}.Result
    Alpha=ScalarParameter() {defaultValue=1, group=Birthday2020, parameterName=BD_Masked_NormalIntensity, sortPriority=51}
   )
  ) {defaultValue=true, group=Birthday2020, parameterName=Use Simple Normal for alt Axis, sortPriority=59}
  n41 = ComponentMask(None=@n30) {b=true, desc=vehicle}
  n42 = ScalarParameter() {defaultValue=10, parameterName=BlendContrast}
  n43 = MaterialFunctionCall(In=@n41, Contrast=@n42) {materialFunction=&CheapContrast}
  n44 = ComponentMask(None=@n30) {desc=Mech, g=true}
  n45 = MaterialFunctionCall(In=@n44, Contrast=@n42) {desc=Mech, materialFunction=&CheapContrast}
  n46 = Saturate(None=Add(A=@n43.Result, B=@n45.Result))
  n47 = MaterialFunctionCall(UVs=ComponentMask(None=@n19) {g=true, r=true}, "Rotation Angle (0-1)"=@n25.B) {materialFunction=&CustomRotator}
  n48 = TextureSample(UVs=@n47.Rotated Values, Tex=@n27) {samplerType=SAMPLERTYPE_Normal}
  n49 = AppendVector(A=ComponentMask(None=@n48.RGB) {g=true, r=true}, B=Multiply(A=@n48.B, B=@n41))
  n50 = Add(
   A=RotateAboutAxis(
    NormalizedRotationAxis=Constant3Vector() {constant=<rgba 0 0 1 0>}
    RotationAngle=Multiply(A=@n25.B) {constB=-1}
    PivotPoint=Constant3Vector() {constant=<rgba 0 0 0 1>}
    Position=@n49
   )
   B=@n49
  )
  n51 = MaterialFunctionCall(In=@n31, Contrast=@n42) {materialFunction=&CheapContrast}
  n52 = LinearInterpolate(
   A=Saturate(None=Add(A=@n51.Result, B=@n45.Result))
   B=Constant()
   Alpha=ScalarParameter() {group=Birthday2020, parameterName=Use Vehicle Mask - For Roll, sortPriority=60}
  )
  n53 = MaterialFunctionCall(UVs=ComponentMask(None=@n19) {b=true, r=true}, "Rotation Angle (0-1)"=@n25.G) {materialFunction=&CustomRotator}
  n54 = TextureSample(UVs=@n53.Rotated Values, Tex=@n27) {samplerType=SAMPLERTYPE_Normal}
  n55 = Multiply(A=@n54.R) {constB=-1}
  n56 = AppendVector(A=AppendVector(A=LinearInterpolate(A=@n54.R, B=@n55), B=Multiply(A=@n54.B, B=@n44)), B=Multiply(A=@n54.G) {constB=-1})
  n57 = Add(
   A=RotateAboutAxis(
    NormalizedRotationAxis=Constant3Vector() {constant=<rgba 0 -1 0 0>}
    RotationAngle=Multiply(A=@n25.G) {constB=-1}
    PivotPoint=Constant3Vector() {constant=<rgba 0 0 0 1>}
    Position=@n56
   )
   B=@n56
  )
  n58 = Saturate(None=Add(A=@n43.Result, B=@n51.Result))
  n59 = LinearInterpolate(
   A=LinearInterpolate(A=@n34, B=@n40, Alpha=@n46)
   B=LinearInterpolate(A=LinearInterpolate(A=@n50, B=@n40, Alpha=@n52), B=LinearInterpolate(A=@n57, B=@n40, Alpha=@n58), Alpha=@n5) {desc=vehicle side}
   Alpha=@n7
  ) {desc=vehicle}
  n60 = LinearInterpolate(A=@n34, B=@n57, Alpha=@n46)
  n61 = Multiply(A=Constant3Vector() {constant=<rgba 0 1 0 0>}, B=@n31)
  n62 = Multiply(A=Constant3Vector() {constant=<rgba 1 0 0 0>}, B=@n44)
  n63 = LinearInterpolate(A=@n61, B=@n62, Alpha=@n7)
  n64 = Multiply(A=Constant3Vector() {constant=<rgba 0 0 1 0>}, B=@n41)
  n65 = Transform(None=Subtract(A=LinearInterpolate(A=@n59, B=@n60, Alpha=@n9) {desc=mech}, B=LinearInterpolate(A=@n63, B=@n64, Alpha=@n9))) {transformSourceType=TRANSFORMSOURCE_Local, transformType=TRANSFORM_Tangent}
  n66 = Constant3Vector() {constant=<rgba 1 1 0 0>}
  n67 = LinearInterpolate(
   A=@n36.Normal
   B=Add(A=@n36.Normal, B=Multiply(A=@n65, B=@n66))
   Alpha=ScalarParameter() {defaultValue=1, group=Birthday2020, parameterName=BD_NormalIntensity, sortPriority=4}
  )
  n68 = TextureObjectParameter() {group=Birthday2020, parameterName=BD_BaseColor_01, sortPriority=12, texture=&T_BlacksRGB}
  n69 = TextureSample(UVs=@n26.Rotated Values, Tex=@n68)
  n70 = AppendVector(A=@n69.RGB, B=@n69.RGB)
  n71 = VectorParameter() {
   group=Birthday2020
   parameterName="BD_Fill_alternate_Axis_Color_A=Roughness 01"
   sortPriority=19
   defaultValue=<rgba 0.2582 0.7605 0.6376 0.3>
  }
  n72 = AppendVector(A=@n71.RGB, B=@n71.RGB)
  n73 = TextureSample(UVs=@n47.Rotated Values, Tex=@n68)
  n74 = LinearInterpolate(A=AppendVector(A=@n73.RGB, B=@n73.RGB), B=@n72, Alpha=@n52)
  n75 = TextureSample(UVs=@n53.Rotated Values, Tex=@n68)
  n76 = AppendVector(A=@n75.RGB, B=@n75.RGB)
  n77 = LinearInterpolate(A=@n76, B=@n72, Alpha=@n58)
  n78 = LinearInterpolate(
   A=LinearInterpolate(A=LinearInterpolate(A=@n70, B=@n72, Alpha=@n46), B=LinearInterpolate(A=@n74, B=@n77, Alpha=@n5) {desc=vehicle side}, Alpha=@n7) {desc=vehicle}
   B=LinearInterpolate(A=@n70, B=@n76, Alpha=@n46)
   Alpha=@n9
  ) {desc=mech}
  n79 = SetMaterialAttributes!(
   MaterialAttributes=@n36.MaterialAttributes
   BaseColor=ComponentMask(None=@n78) {b=true, g=true, r=true}
   Roughness=LinearInterpolate(
    A=ScalarParameter() {group=Birthday2020, parameterName=BD_Roughness_Min, sortPriority=8}
    B=ScalarParameter() {defaultValue=1, group=Birthday2020, parameterName=BD_Roughness_Max, sortPriority=9}
    Alpha=ComponentMask(None=@n78) {a=true}
   )
   Normal=ShadingPathSwitch(Default=@n67, Deferred=@n67, Mobile=@n36.Normal) {desc=Migrated from FeatureLevelSwitch}
   Specular=ScalarParameter() {group=Birthday2020, parameterName=BD_Specular, sortPriority=5}
   Metallic=ScalarParameter() {group=Birthday2020, parameterName=BD_Metallic, sortPriority=7}
   EmissiveColor
  ) {attributeSetTypes [
    #1
    #2
    #3
    #4
    #5
    #6
   ]}
  n80 = FunctionInput() {
   bUsePreviewValueAsDefault=true
   desc=Secondary
   inputName="Customization Mask 2 (_CM B Top Right)"
   inputType=FunctionInput_Scalar
   sortPriority=25
  }
  n81 = FunctionInput() {
   bUsePreviewValueAsDefault=true
   desc=Base Customization
   inputName="Customization Mask (_CM B Top left)"
   inputType=FunctionInput_Scalar
   sortPriority=24
  }
  n82 = Add(A=@n80, B=@n81) {desc="Base + Secondary = exposed primary"}
  n83 = BlendMaterialAttributes(A=@n36.MaterialAttributes, B=@n79, Alpha=@n82)
  n84 = GetMaterialAttributes(None=@n83) {attributeGetTypes [
    #1
    #5
    #4
    #2
    #3
   ]}
  n85 = Multiply(
   A=TextureCoordinate()
   B=LinearInterpolate(
    A=ScalarParameter() {defaultValue=6, group=Birthday2020_Secondary, parameterName=BD_SecondaryScale_Weapon, sortPriority=1}
    B=ScalarParameter() {defaultValue=4, group=Birthday2020_Secondary, parameterName=BD_SecondaryScale_Vehicle, sortPriority=2}
    Alpha=@n7
   )
  )
  n86 = TextureSampleParameter2D(UVs=@n85) {
   group=Birthday2020_Secondary
   parameterName=BD_Secondary_BaseColor_01
   sortPriority=-10
   texture=&T_BlacksRGB
  }
  n87 = TextureSampleParameter2D(UVs=@n85) {
   group=Birthday2020_Secondary
   parameterName=BD_SecondaryNormal_02
   samplerType=SAMPLERTYPE_Normal
   sortPriority=0
   texture=&FlatNormalBlue
  }
  n88 = LinearInterpolate(
   A=@n36.Normal
   B=MaterialFunctionCall(BaseNormal=@n36.Normal, AdditionalNormal=@n87.RGB) {materialFunction=&BlendAngleCorrectedNormals}.Result
   Alpha=ScalarParameter() {defaultValue=1, group=Birthday2020_Secondary, parameterName=BD_Secondary_NormalIntensity, sortPriority=4}
  )
  n89 = SetMaterialAttributes!(
   MaterialAttributes=@n84.MaterialAttributes
   BaseColor=@n86.RGB
   Metallic=ScalarParameter() {group=Birthday2020_Secondary, parameterName=BD_Secondary_Metallic, sortPriority=-4}
   Specular=ScalarParameter() {defaultValue=0.5, group=Birthday2020_Secondary, parameterName=BD_Secondary_Specular, sortPriority=-4}
   Roughness=@n86.RGB
   Normal=ShadingPathSwitch(Default=@n88, Deferred=@n88, Mobile=@n36.Normal) {desc=Migrated from FeatureLevelSwitch}
  ) {attributeSetTypes [
    #1
    #5
    #4
    #2
    #3
   ]}
  n90 = GetMaterialAttributes(None=BlendMaterialAttributes(A=@n83, B=@n89, Alpha=@n80)) {attributeGetTypes [#6]}
  n91 = Constant() {r=0.5}
  n92 = Time()
  n93 = ScalarParameter() {defaultValue=-9999999, group=Birthday2020_Reactive, parameterName=BD_ConfettiB_Opacity_Time}
  n94 = ScalarParameter() {defaultValue=0.65, group=Birthday2020_Reactive, parameterName=BD_Confetti_Opacity_Duration}
  n95 = Saturate(None=Divide(A=Subtract(A=@n92, B=@n93), B=@n94))
  n96 = ScalarParameter() {defaultValue=0.5, group=Birthday2020_Reactive, parameterName=BD_Confetti_Opacity_Curve}
  n97 = ScalarParameter() {group=Birthday2020_Reactive, parameterName=BD_TestReactive, sortPriority=300}
  n98 = LinearInterpolate(A=OneMinus(None=Power(Base=@n95, Exp=@n96)), B=@n97, Alpha=@n97)
  n99 = Constant3Vector() {constant=<rgba 0 0 -0.25 1>}
  n100 = Constant3Vector() {constant=<rgba 0 0 0 1>}
  n101 = VectorParameter() {group=Birthday2020, parameterName="BD_Confetti_alternate_Axis_Color_A=Roughness", sortPriority=43}
  n102 = LinearInterpolate(
   A=LinearInterpolate(
    A=ScalarParameter() {defaultValue=1, group=Birthday2020, parameterName=BD_ConfettiB_Weapon_Scale, sortPriority=40}
    B=ScalarParameter() {defaultValue=1, group=Birthday2020, parameterName=BD_ConfettiB_Vehicle_Scale, sortPriority=40}
    Alpha=@n7
   )
   B=ScalarParameter() {defaultValue=1, group=Birthday2020, parameterName=BD_ConfettiB_Mech_Scale, sortPriority=40}
   Alpha=@n9
  )
  n103 = Time()
  n104 = ScalarParameter() {defaultValue=-9999999, group=Birthday2020_Reactive, parameterName=BD_ConfettiB_Main_Time, sortPriority=40}
  n105 = Saturate(None=Divide(
    A=Subtract(A=@n103, B=@n104)
    B=ScalarParameter() {defaultValue=0.5, group=Birthday2020_Reactive, parameterName=BD_Confetti_Main_Duration, sortPriority=41}
   ))
  n106 = ScalarParameter() {defaultValue=0.6, group=Birthday2020_Reactive, parameterName=BD_Confetti_Main_Curve, sortPriority=42}
  n107 = Power(Base=Multiply(A=@n105) {constB=0.25}, Exp=@n106)
  n108 = OneMinus(None=@n107)
  n109 = Power(Base=@n105, Exp=@n106)
  n110 = LinearInterpolate(
   A=MaterialFunctionCall(G=@n108, B=@n109) {materialFunction=&AppendMany}.RGB
   B=MaterialFunctionCall(R=@n107, G=@n108) {materialFunction=&AppendMany}.RGB
   Alpha=@n7
  )
  n111 = MaterialFunctionCall(R=@n108, B=@n109) {materialFunction=&AppendMany}
  n112 = TextureObjectParameter() {group=Birthday2020, parameterName=BD_Mask_02, sortPriority=11, texture=&T_BlacksRGB}
  n113 = MaterialFunctionCall!(
   IsCharacter
   IsStatsB
   Texture=@n112
   Rotation_Axis=LinearInterpolate(A=LinearInterpolate(A=@n99, B=@n100, Alpha=@n7), B=Constant3Vector() {constant=<rgba 0.25 0.25 0 1>}, Alpha=@n9)
   Caculated_Bounds=Add(A=Multiply(A=@n13, B=@n102), B=LinearInterpolate(A=@n110, B=@n111.RGB, Alpha=@n9))
   Pre_Skinned_Normals=@n29
   Blend_Contrast=ScalarParameter() {group=Birthday2020, parameterName=BD_Confetti_Blend_Contrast, sortPriority=41}
   Fill_alternate_Axis_Opacity=ScalarParameter() {group=Birthday2020, parameterName=BD_Confetti_Fill_alternate_Axis_Opacity, sortPriority=42}
   "Fill_alternate_Axis_Color_A=Roughness"=AppendVector(A=@n101.RGB, B=@n101.RGB)
   Is_Vehicle=@n7
   Is_Mech=@n9
   Enable_Fill_Z_Weapon
   Fill_Z_Weapon_opacity
   Fill_Z_Weapon_Color
  ) {materialFunction=&/MF_TopsSideProjection_Base}
  n114 = ChannelMaskParameter(None=@n113.Combined) {
   desc=uses CarnB_Mask for source
   group=Birthday2020_Confetti
   maskChannel=Green
   parameterName=BD_ConfettiB_Color_Mask
   defaultValue=<rgba 0 1 0 0>
  }
  n115 = ScalarParameter() {defaultValue=-3, group=Birthday2020_Confetti, parameterName=BD_ConfettiB_Offset}
  n116 = ScalarParameter() {defaultValue=1.75, group=Birthday2020_Confetti, parameterName=BD_ConfettiB_Falloff}
  n117 = VectorParameter() {group=Birthday2020_Gradient, parameterName=BD_ConfettiA_Spread, sortPriority=30, defaultValue=<rgba 1 1 1 0>}
  n118 = ScalarParameter() {defaultValue=1, group=Birthday2020_Gradient, parameterName=BD_Texture_Rotate_Mech, sortPriority=13}
  n119 = ScalarParameter() {defaultValue=1, group=Birthday2020_Gradient, parameterName=BD_Texture_Rotate_Vehicle_Side, sortPriority=13}
  n120 = ScalarParameter() {defaultValue=1, group=Birthday2020_Gradient, parameterName=BD_Texture_Rotate_Vehicle, sortPriority=13}
  n121 = FunctionInput(Preview=Constant3Vector() {constant=<rgba 0 1 0 0>}) {bUsePreviewValueAsDefault=true, inputName=ForwardAxis, sortPriority=5}
  n122 = MaterialFunctionCall!(
   ForwardAxis=@n121
   Bounds=@n13
   Pre-Skinned Local Position
   isVehicle=@n9
   isMech=@n7
   Side Project Vehicle
   WeaponOffset=@n115
   WeaponFalloff=@n116
   VehicleOffset=@n115
   VehicleFalloff=@n116
   MechOffset=@n115
   MechFalloff=@n116
   Invert_Gradent
   UV_Spread_XY=@n117.RGB
   UV_Animated_XY
   Warp_UVs
   Rotate_Mech=@n118
   Rotate_Vehicle_Side=@n119
   Rotate_Vehicle=@n120
  ) {materialFunction=&/MF_Wrap_Gradent_Creation}
  n123 = TextureObjectParameter() {
   group=Birthday2020_Gradient
   parameterName=BD_Emissive01Gradent_Texture
   samplerType=SAMPLERTYPE_Masks
   sortPriority=12
   texture=/Game/Global/Textures/Default/Masks_and_Gradients/T_VertGradient_MASK
  }
  n124 = Saturate(None=MaterialFunctionCall(
    Transition Phase=TextureSample(UVs=@n122.Uvs, Tex=@n123).R
    Height Texture=ChannelMaskParameter(None=@n113.Combined) {desc=uses CarnB_Mask for source, group=Birthday2020_Confetti, parameterName=BD_ConfettiB_Mask_Height}.R
    Contrast=ScalarParameter() {defaultValue=5, group=Birthday2020_Confetti, parameterName=BD_ConfettiB_Contrast_Height}
   ) {materialFunction=&HeightLerp}.Alpha)
  n125 = VectorParameter() {group=Birthday2020_Confetti, parameterName=BD_ConfettiB_ColorReactive, defaultValue=<rgba 0.1095 0.9911 0.9216 1>}
  n126 = MaterialFunctionCall(Hue Shift Percentage=@n114.R, Texture=Multiply(A=@n124, B=@n125.RGB)) {materialFunction=&HueShift}
  n127 = Divide(
   A=Subtract(A=Time(), B=ScalarParameter() {defaultValue=-9999999, group=Birthday2020_Reactive, parameterName=BD_ConfettiA_Opacity_Time})
   B=@n94
  )
  n128 = OneMinus(None=Power(Base=Saturate(None=@n127), Exp=@n96))
  n129 = ScalarParameter() {group=Birthday2020_Reactive, parameterName=BD_TestReactive, sortPriority=300}
  n130 = VectorParameter() {group=Birthday2020, parameterName="BD_Confetti_Fill_alternate_Axis_Color_A=Roughness", sortPriority=30}
  n131 = Time()
  n132 = ScalarParameter() {defaultValue=-9999999, group=Birthday2020_Reactive, parameterName=BD_ConfettiA_Main_Time, sortPriority=30}
  n133 = Saturate(None=Divide(
    A=Subtract(A=@n131, B=@n132)
    B=ScalarParameter() {defaultValue=0.5, group=Birthday2020_Reactive, parameterName=BD_Confetti_Main_Duration, sortPriority=41}
   ))
  n134 = ScalarParameter() {defaultValue=0.6, group=Birthday2020_Reactive, parameterName=BD_Confetti_Main_Curve, sortPriority=42}
  n135 = Power(Base=Multiply(A=@n133) {constB=0.25}, Exp=@n134)
  n136 = Power(Base=@n133, Exp=@n134)
  n137 = LinearInterpolate(
   A=MaterialFunctionCall(G=@n135, B=@n136) {materialFunction=&AppendMany}.RGB
   B=MaterialFunctionCall(R=@n136, G=@n135) {materialFunction=&AppendMany}.RGB
   Alpha=@n7
  )
  n138 = MaterialFunctionCall(R=@n135, B=@n136) {materialFunction=&AppendMany}
  n139 = LinearInterpolate(
   A=LinearInterpolate(
    A=ScalarParameter() {defaultValue=1, group=Birthday2020, parameterName=BD_ConfettiA_Weapon_Scale, sortPriority=30}
    B=ScalarParameter() {defaultValue=1, group=Birthday2020, parameterName=BD_ConfettiA_Vehicle_Scale, sortPriority=30}
    Alpha=@n7
   )
   B=ScalarParameter() {defaultValue=1, group=Birthday2020, parameterName=BD_ConfettiA_Mech_Scale, sortPriority=30}
   Alpha=@n9
  )
  n140 = Constant3Vector() {constant=<rgba 0 0 -0.25 1>}
  n141 = Constant3Vector() {constant=<rgba 0 0 0 1>}
  n142 = MaterialFunctionCall(
   Texture=@n112
   Rotation_Axis=LinearInterpolate(A=LinearInterpolate(A=@n140, B=@n141, Alpha=@n7), B=Constant3Vector() {constant=<rgba 0.25 0.25 0 1>}, Alpha=@n9)
   Caculated_Bounds=Add(A=Multiply(A=@n13, B=@n139), B=LinearInterpolate(A=@n137, B=@n138.RGB, Alpha=@n9))
   Pre_Skinned_Normals=@n29
   Blend_Contrast=ScalarParameter() {group=Birthday2020, parameterName=BD_Confetti_Blend_Contrast, sortPriority=30}
   Fill_alternate_Axis_Opacity=ScalarParameter() {group=Birthday2020, parameterName=BD_Confetti_Fill_alternate_Axis_Opacity, sortPriority=30}
   "Fill_alternate_Axis_Color_A=Roughness"=AppendVector(A=@n130.RGB, B=@n130.RGB)
   Is_Vehicle=@n7
   Is_Mech=@n9
  ) {materialFunction=&MF_TopsSideProjection_Base}
  n143 = ChannelMaskParameter(None=@n142.Combined) {
   desc=uses ConfettiB_Mask for source
   group=Birthday2020_Confetti
   maskChannel=Alpha
   parameterName=BD_ConfettiA_Color_Mask
   sortPriority=30
   defaultValue=<rgba 0 0 0 1>
  }
  n144 = ScalarParameter() {defaultValue=-3, group=Birthday2020_Confetti, parameterName=BD_ConfettiA_Offset, sortPriority=30}
  n145 = ScalarParameter() {defaultValue=1.75, group=Birthday2020_Confetti, parameterName=BD_ConfettiA_Falloff, sortPriority=30}
  n146 = MaterialFunctionCall(
   ForwardAxis=@n121
   Bounds=@n13
   isVehicle=@n7
   isMech=@n9
   WeaponOffset=@n144
   WeaponFalloff=@n145
   VehicleOffset=@n144
   VehicleFalloff=@n145
   MechOffset=@n144
   MechFalloff=@n145
   UV_Spread_XY=@n117.RGB
   Rotate_Mech=@n118
   Rotate_Vehicle_Side=@n119
   Rotate_Vehicle=@n120
  ) {materialFunction=&MF_Wrap_Gradent_Creation}
  n147 = Saturate(None=MaterialFunctionCall(
    Transition Phase=TextureSample(UVs=@n146.Uvs, Tex=@n123).R
    Height Texture=ChannelMaskParameter(None=@n142.Combined) {desc=uses ConfettiB_Mask for source, group=Birthday2020_Confetti, parameterName=BD_ConfettiA_Mask_Height, sortPriority=30}.R
    Contrast=ScalarParameter() {defaultValue=5, group=Birthday2020_Confetti, parameterName=BD_ConfettiA_Contrast_Height, sortPriority=30}
   ) {materialFunction=&HeightLerp}.Alpha)
  n148 = VectorParameter() {group=Birthday2020_Confetti, parameterName=BD_ConfettiA_ColorReactive, sortPriority=30, defaultValue=<rgba 1 1 0.0666 1>}
  n149 = Multiply(A=@n147, B=@n148.RGB)
  n150 = Multiply(
   A=LinearInterpolate(
    A=LinearInterpolate(B=@n126.Result, Alpha=@n98)
    B=MaterialFunctionCall(Hue Shift Percentage=@n143.R, Texture=@n149) {materialFunction=&HueShift}.Result
    Alpha=LinearInterpolate(A=@n128, B=@n129, Alpha=@n129)
   )
   B=MaterialFunctionCall(Value_Morning=@n91, Value_Day=Constant() {r=1}, Value_Evening=@n91, Value_Night=Constant() {r=0.25}) {materialFunction=&Scalar_Time_of_Day}.Result
  )
  n151 = StaticSwitchParameter(
   True=BlendMaterialAttributes(
    A=@n83
    B=SetMaterialAttributes!(MaterialAttributes=@n90.MaterialAttributes, EmissiveColor=@n150) {attributeSetTypes [#6]}
    Alpha=Constant() {r=1}
   )
   False=@n35
  ) {
   desc="Side Projection Base - Alt and Z masking (Normal map enabled) - Reactive -Topside Emissive 2 parallax levels - TOD emissive restrictive"
   group=Customization_Wrap
   parameterName=Wrap_Birthday2020
  }
  FunctionOutput(None=@n151)
  n152 = MaterialFunctionCall(Color=@n19) {materialFunction=&SplitComponents}
  n153 = AppendVector(A=@n152.G, B=@n152.G)
  n154 = LinearInterpolate(A=AppendVector(A=@n152.R, B=@n152.R), B=AppendVector(A=@n152.R, B=@n152.R), Alpha=@n5)
  n155 = OneMinus(None=@n152.B)
  n156 = TextureSampleParameter2D(UVs=LinearInterpolate(A=LinearInterpolate(A=@n153, B=@n154, Alpha=@n7), B=AppendVector(A=@n155, B=@n152.R), Alpha=@n9)) {
   group=Birthday2020
   parameterName=BD_MaskTexture_01
   samplerType=SAMPLERTYPE_Masks
   sortPriority=-9
   texture=&T_BlackMask
  }
  LinearInterpolate(
   A=ScalarParameter() {group=Birthday2020, parameterName=BD_Specular_Min, sortPriority=5}
   B=ScalarParameter() {defaultValue=1, group=Birthday2020, parameterName=BD_Specular_Max, sortPriority=6}
   Alpha=ChannelMaskParameter(None=@n156.R) {group=Birthday2020, parameterName=BD_SpecChannel_Mask_01, sortPriority=5}.R
  )
  LinearInterpolate(A=@n61, B=@n64, Alpha=@n58)
  LinearInterpolate(
   A=LinearInterpolate(A=@n81, B=@n82, Alpha=ScalarParameter() {group=Wrap_CosmicPulse, parameterName="CP_BaseAndSecondaryMasked=ExposedPrimary", sortPriority=-99})
   B=Add(A=OneMinus(None=@n82) {desc=Primary}, B=@n81)
   Alpha=ScalarParameter() {group=Wrap_CosmicPulse, parameterName="CP_BaseAndPrimaryMasked=ExposedSecondary", sortPriority=-99}
  )
  LinearInterpolate(A=@n62, B=@n64, Alpha=@n46)
  LinearInterpolate(A=@n62, B=@n61, Alpha=@n46)
  MaterialFunctionCall(BaseNormal=@n36.Normal) {materialFunction=&BlendAngleCorrectedNormals}
  n157 = VectorParameter() {group=Birthday2020_Confetti, parameterName=BD_ConfettiA_Color, sortPriority=30, defaultValue=<rgba 1 1 0.0666 1>}
  Multiply(A=@n147, B=@n157.RGB)
  n158 = VectorParameter() {group=Birthday2020_Confetti, parameterName=BD_ConfettiB_Color, defaultValue=<rgba 0.1095 0.9911 0.9216 1>}
  Multiply(A=@n124, B=@n158.RGB)
  Multiply(A=@n48.RGB, B=Constant3Vector() {constant=<rgba -1 -1 1 0>})
  n159 = Multiply(A=ScalarParameter() {group=Birthday2020, parameterName=BEATS_Emissive01-DarkenBase-------, sortPriority=-20}, B=Add(A=@n147, B=@n124))
  Multiply(B=OneMinus(None=Saturate(None=@n159)))
  Multiply(A=@n19, B=Constant3Vector() {constant=<rgba 1 1 -1 0>})
  Reroute(None=@n29)
 ]}
