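# Material function that projects a tiling normal map triplanar-style, using a position input (world position by default) and the pre-skinned normal.
# Position over TextureSize feeds three planar samples with per-axis rotation, with X and Z results rotated back to match.
# Contrast-adjusted normal components blend the projections.
# It subtracts an equally blended axis-aligned base normal and transforms the difference from local to tangent space, outputting a detail offset rather than a full normal.
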MaterialFunction /Game/Packages/Fortress_SharedMaterials/Base_Material_Functions/BaseCustomizationFunctions/PreSkinnedAlignedNormal
graph {
 properties {bExposeToLibrary=true, description=Tiles a texture in worldspace, libraryCategoriesText [Misc,Texturing]}
 nodes [
  AppendVector(B=Constant() {r=1})
  n1 = FunctionInput() {bUsePreviewValueAsDefault=true, description=Size of texture in worldspace units, inputName=TextureSize, previewValue=<64 64 64 6>}
  n2 = Divide(
   A=FunctionInput(Preview=WorldPosition() {worldPositionShaderOffset=WPT_ExcludeAllShaderOffsets}.XYZ) {bUsePreviewValueAsDefault=true, inputName=WorldPosition, sortPriority=100}
   B=Multiply(A=Abs(None=@n1)) {constA=-1, constB=-1}
  )
  n3 = FunctionInput(Preview=TextureObject() {samplerType=SAMPLERTYPE_Normal, texture=/Engine/Functions/Engine_MaterialFunctions02/ExampleContent/Textures/SphereRenderNormalsMap}) {inputName=TextureObject, inputType=FunctionInput_Texture2D}
  n4 = TextureSample(
   UVs=MaterialFunctionCall(
    UVs=ComponentMask(None=@n2) {b=true, r=true}
    "Rotation Angle (0-1)"=FunctionInput() {bUsePreviewValueAsDefault=true, inputName=RotYAxis, inputType=FunctionInput_Scalar}
   ) {materialFunction=&CustomRotator}.Rotated Values
   Tex=@n3
  ) {samplerSource=SSM_Wrap_WorldGroupSettings, samplerType=SAMPLERTYPE_Normal}
  n5 = Multiply(A=@n4.R) {constB=-1}
  n6 = PreSkinnedNormal()
  n7 = Multiply(
   A=FunctionInput(Preview=VertexInterpolator(VS=@n6).PS) {bUsePreviewValueAsDefault=true, inputName=PreSkinNormal, previewValue=<0 0 1 1>}
   B=Constant() {r=1}
  )
  n8 = ComponentMask(None=@n7) {g=true}
  n9 = Multiply(A=@n4.B, B=@n8)
  n10 = FunctionInput() {bUsePreviewValueAsDefault=true, inputName=RotXAxis, inputType=FunctionInput_Scalar}
  n11 = TextureSample(
   UVs=MaterialFunctionCall(UVs=ComponentMask(None=@n2) {b=true, g=true}, "Rotation Angle (0-1)"=@n10) {materialFunction=&CustomRotator}.Rotated Values
   Tex=@n3
  ) {samplerSource=SSM_Wrap_WorldGroupSettings, samplerType=SAMPLERTYPE_Normal}
  n12 = ComponentMask(None=@n7) {r=true}
  n13 = AppendVector(A=AppendVector(A=Multiply(A=@n12, B=@n11.B), B=Multiply(A=@n11.R) {constB=-1}), B=Multiply(A=@n11.G) {constB=-1})
  n14 = Add(
   A=RotateAboutAxis(
    NormalizedRotationAxis=Constant3Vector() {constant=<rgba 1 0 0 0>}
    RotationAngle=Multiply(A=@n10) {constB=-1}
    PivotPoint=Constant3Vector()
    Position=@n13
   )
   B=@n13
  )
  n15 = FunctionInput() {bUsePreviewValueAsDefault=true, inputName=ProjectionTransitionContrast, inputType=FunctionInput_Scalar, sortPriority=50}
  n16 = MaterialFunctionCall(In=Abs(None=@n12), Contrast=@n15) {materialFunction=&CheapContrast}
  n17 = LinearInterpolate(A=AppendVector(A=AppendVector(A=@n5, B=@n9), B=Multiply(A=@n4.G) {constB=-1}), B=@n14, Alpha=@n16.Result)
  n18 = FunctionInput() {bUsePreviewValueAsDefault=true, inputName=RotZAxis, inputType=FunctionInput_Scalar}
  n19 = TextureSample(
   UVs=MaterialFunctionCall(UVs=ComponentMask(None=@n2) {g=true, r=true}, "Rotation Angle (0-1)"=@n18) {materialFunction=&CustomRotator}.Rotated Values
   Tex=@n3
  ) {samplerSource=SSM_Wrap_WorldGroupSettings, samplerType=SAMPLERTYPE_Normal}
  n20 = Constant3Vector() {constant=<rgba -1 -1 1 0>}
  n21 = ComponentMask(None=@n7) {b=true}
  n22 = AppendVector(A=ComponentMask(None=Multiply(A=@n19.RGB, B=@n20)) {g=true, r=true}, B=Multiply(A=@n19.B, B=@n21))
  n23 = Add(
   A=RotateAboutAxis(
    NormalizedRotationAxis=Constant3Vector() {constant=<rgba 0 0 1 0>}
    RotationAngle=Multiply(A=@n18) {constB=-1}
    PivotPoint=Constant3Vector()
    Position=@n22
   )
   B=@n22
  )
  n24 = MaterialFunctionCall(In=Abs(None=@n21), Contrast=@n15) {materialFunction=&CheapContrast}
  n25 = LinearInterpolate(
   A=Multiply(A=Constant3Vector() {constant=<rgba 0 1 0 0>}, B=@n8)
   B=Multiply(A=Constant3Vector() {constant=<rgba 1 0 0 0>}, B=@n12)
   Alpha=@n16.Result
  )
  n26 = Multiply(A=Constant3Vector() {constant=<rgba 0 0 1 0>}, B=@n21)
  n27 = Subtract(
   A=Normalize(VectorInput=LinearInterpolate(A=@n17, B=@n23, Alpha=@n24.Result))
   B=Normalize(VectorInput=LinearInterpolate(A=@n25, B=@n26, Alpha=@n24.Result))
  )
  FunctionOutput(None=Transform(None=@n27) {transformSourceType=TRANSFORMSOURCE_Local, transformType=TRANSFORM_Tangent}) {outputName=XYZ Texture, sortPriority=1}
  Normalize(VectorInput=@n14)
  StaticComponentMaskParameter(None=@n27) {defaultG=true, defaultR=true}
 ]
}
