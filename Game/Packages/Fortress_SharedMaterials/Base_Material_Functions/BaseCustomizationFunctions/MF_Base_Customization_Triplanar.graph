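# Material function for base customization that triplanar-projects diffuse, normal and mask textures from pre-skinned local position and normal.
# Flags for pickaxe, character and backpack select per-type scale and rotation parameters for the offset position and projection axes, with extra rotation for mechs.
# WorldAlignedTexture supplies diffuse, roughness (from diffuse alpha) and mask.
# PreSkinnedAlignedNormal adds only its XY onto the material normals, weighted by an intensity input.
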MaterialFunction /Game/Packages/Fortress_SharedMaterials/Base_Material_Functions/BaseCustomizationFunctions/MF_Base_Customization_Triplanar
graph {nodes [
  Constant3Vector() {constant=<rgba 0 0 0.5 0>}
  Divide()
  n1 = StaticBool()
  n2 = FunctionInput(Preview=@n1) {bUsePreviewValueAsDefault=true, description="Default = False", inputName=Export Float 4, inputType=FunctionInput_StaticBool, sortPriority=4}
  n3 = FunctionInput() {
   bUsePreviewValueAsDefault=true
   description=Size of texture in worldspace units
   inputName=TextureScale
   sortPriority=-18
   previewValue=<64 64 64 6>
  }
  n4 = Add(
   A=FunctionInput() {bUsePreviewValueAsDefault=true, inputName=Triplanner_Offset}
   B=FunctionInput() {bUsePreviewValueAsDefault=true, inputName=Pre-Skinned Local Position, sortPriority=-16}
  )
  n5 = VectorParameter() {parameterName=WorldPositionScaleAdjust_Pliant_Pickaxe, defaultValue=<rgba 1 1 1 0>}
  n6 = VectorParameter() {parameterName=WorldPositionScaleAdjust_Pliant_Character, defaultValue=<rgba 1 1 1 0>}
  n7 = VectorParameter() {parameterName=WorldPositionScaleAdjust_Pliant_Backpack, defaultValue=<rgba 1 1 1 0>}
  n8 = LinearInterpolate(
   A=LinearInterpolate(A=@n4, B=Multiply(A=@n4, B=@n5.RGB), Alpha=NamedRerouteUsage())
   B=LinearInterpolate(A=Multiply(A=@n4, B=@n6.RGB), B=Multiply(A=@n4, B=@n7.RGB), Alpha=NamedRerouteUsage())
   Alpha=NamedRerouteUsage()
  ) {constAlpha=1}
  n9 = FunctionInput() {bUsePreviewValueAsDefault=true, inputName=Pre-Skinned Local Normal, sortPriority=-15}
  n10 = FunctionInput() {
   bUsePreviewValueAsDefault=true
   inputName=ProjectionTransitionContrast
   inputType=FunctionInput_Scalar
   sortPriority=10
   previewValue=<1 0 0 1>
  }
  n11 = FunctionInput() {bUsePreviewValueAsDefault=true, inputName=Rotation Z Axis, inputType=FunctionInput_Scalar}
  n12 = ScalarParameter() {defaultValue=0.25, parameterName=Mech Rotation}
  n13 = FunctionInput() {bUsePreviewValueAsDefault=true, inputName=isMech, inputType=FunctionInput_Scalar}
  n14 = FunctionInput() {bUsePreviewValueAsDefault=true, inputName=Rotation X Axis, inputType=FunctionInput_Scalar}
  n15 = MaterialFunctionCall(
   X=LinearInterpolate(A=@n14, B=Add(A=@n14, B=@n12), Alpha=@n13)
   Y=LinearInterpolate(A=@n11, B=Add(A=@n11, B=@n12), Alpha=@n13)
   Z=LinearInterpolate(A=Constant(), B=Add(A=@n14, B=@n12), Alpha=@n13)
  ) {materialFunction=&MakeFloat3}
  n16 = Add(A=@n15.Result, B=VectorParameter() {parameterName=Rotation_Pliant_Pickaxe}.RGB)
  n17 = NamedRerouteUsage()
  n18 = Add(A=@n15.Result, B=VectorParameter() {parameterName=Rotation_Pliant_Character}.RGB)
  n19 = Add(A=@n15.Result, B=VectorParameter() {parameterName=Rotation_Pliant_Backpack}.RGB)
  n20 = NamedRerouteUsage()
  n21 = MaterialFunctionCall(Float3=LinearInterpolate(A=LinearInterpolate(A=@n15.Result, B=@n16, Alpha=@n17), B=LinearInterpolate(A=@n18, B=@n19, Alpha=@n20), Alpha=NamedRerouteUsage())) {materialFunction=&BreakOutFloat3Components}
  n22 = MaterialFunctionCall(
   Export Float 4=@n2
   TextureObject=FunctionInput(Preview=TextureObject() {texture=/Engine/EngineMaterials/DefaultDiffuse}) {bUsePreviewValueAsDefault=true, inputName=TextureObject_Diffuse, inputType=FunctionInput_Texture2D, sortPriority=-20}
   TextureSize=@n3
   WorldPosition=@n8
   World Space Normal=@n9
   RotZAxis=@n21.G
   RotXAxis=@n21.R
   RotYAxis=@n21.B
   ProjectionTransitionContrast=@n10
  ) {materialFunction=/Game/Weapons/WeaponSkins/Wraps/MaterialFunctions/WorldAlignedTexture}
  FunctionOutput(None=ComponentMask(None=@n22.XYZ Texture) {b=true, g=true, r=true}) {outputName=Diffuse Texture}
  n23 = TextureSample() {samplerSource=SSM_Wrap_WorldGroupSettings, samplerType=SAMPLERTYPE_Normal, texture=&FlatNormalBlue}
  n24 = ShadingPathSwitch(Default=FunctionInput(Preview=@n23.RGB) {bUsePreviewValueAsDefault=true, inputName=Material Normals, sortPriority=-21}) {desc=Migrated from FeatureLevelSwitch}
  n25 = TextureObject() {samplerType=SAMPLERTYPE_Normal, texture=&FlatNormalBlue}
  n26 = Multiply(
   A=MaterialFunctionCall!(
    TextureObject=FunctionInput(Preview=@n25) {bUsePreviewValueAsDefault=true, inputName=TextureObject_Normal, inputType=FunctionInput_Texture2D, sortPriority=-19}
    TextureSize=@n3
    RotYAxis=@n21.B
    RotZAxis=@n21.G
    RotXAxis=@n21.R
    PreSkinNormal=@n9
    ProjectionTransitionContrast=@n10
    WorldPosition=@n8
   ) {materialFunction=&/PreSkinnedAlignedNormal}.XYZ Texture
   B=Constant3Vector() {constant=<rgba 1 1 0 0>}
  )
  n27 = FunctionInput() {bUsePreviewValueAsDefault=true, inputName=Normal Texture Intensity, inputType=FunctionInput_Scalar, sortPriority=-17, previewValue=<1 0 0 1>}
  FunctionOutput(None=LinearInterpolate(A=@n24, B=Add(A=@n24, B=@n26), Alpha=@n27)) {outputName=Normal Texture, sortPriority=3}
  FunctionOutput(None=MaterialFunctionCall(
    Export Float 4=@n1
    TextureObject=FunctionInput() {bUsePreviewValueAsDefault=true, inputName=TextureObject_MASK, inputType=FunctionInput_Texture2D, sortPriority=-20}
    TextureSize=@n3
    WorldPosition=@n8
    World Space Normal=@n9
    RotZAxis=@n21.G
    RotXAxis=@n21.R
    RotYAxis=@n21.B
    ProjectionTransitionContrast=@n10
   ) {materialFunction=&WorldAlignedTexture}.XYZ Texture) {outputName=Output MASK}
  FunctionOutput(None=ComponentMask(None=@n22.XYZ Texture) {a=true}) {outputName=Roughness Texture, sortPriority=1}
  n28 = FunctionInput() {bUsePreviewValueAsDefault=true, inputName=IsStatsB, inputType=FunctionInput_Vector2, sortPriority=78}
  NamedRerouteDeclaration(None=@n28) {name=Is_StatsB, nodeColor=<rgba 0 0.9253 1 1>}
  NamedRerouteDeclaration(None=ComponentMask(None=@n28) {g=true}) {name=IsPickaxe, nodeColor=<rgba 0 0.9253 1 1>}
  NamedRerouteDeclaration(None=ComponentMask(None=@n28) {r=true}) {name=IsBackpack, nodeColor=<rgba 0 0.9253 1 1>}
  NamedRerouteDeclaration(None=FunctionInput() {bUsePreviewValueAsDefault=true, inputName=isCharacter, inputType=FunctionInput_Scalar, sortPriority=78}) {name=isCharacter, nodeColor=<rgba 0 0.9253 1 1>}
 ]}
