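# Material function for weapon, vehicle and mech wrap skins that blends a triplanar detail normal into incoming material attributes, behind a static switch.
# Per-type scale, offset and speed parameters feed the shared customization triplanar function, with optional noise-texture position warp.
# Offset and time panning are multiplied by bounds Max minus itself, so both evaluate to zero as wired.
# The optionally masked normal blends angle-corrected over the base, weighted by the customization mask.
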MaterialFunction /Game/Weapons/WeaponSkins/Wraps/MaterialFunctions/MF_BlendedPanningNormal
graph {nodes [
  n1 = FunctionInput() {inputName=Material, inputType=FunctionInput_MaterialAttributes, sortPriority=-200}
  n2 = GetMaterialAttributes(None=@n1) {attributeGetTypes [#1]}
  n3 = Constant3Vector() {constant=<rgba 0 0 1 0>}
  n4 = TextureObjectParameter() {
   group=Primary Triplaner BlendedPanningNormals
   parameterName=BPN_Normal
   samplerType=SAMPLERTYPE_Normal
   sortPriority=7
   texture=&FlatNormalBlue
  }
  n5 = ScalarParameter() {defaultValue=60, group=Primary Triplaner BlendedPanningNormals, parameterName=BPN_Weapon_Texture_Scale, sortPriority=2}
  n6 = ScalarParameter() {defaultValue=60, group=Primary Triplaner BlendedPanningNormals, parameterName=BPN_Vehicle_Texture_Scale, sortPriority=3}
  n7 = FunctionInput() {inputName=isVehicle, inputType=FunctionInput_Scalar, sortPriority=20}
  n8 = FunctionInput() {inputName=isMech, inputType=FunctionInput_Scalar, sortPriority=21}
  n9 = ScalarParameter() {defaultValue=1, group=Primary Triplaner BlendedPanningNormals, parameterName=BPN_Triplaners Normal Intenisty, sortPriority=5}
  n10 = VectorParameter() {group=Primary Triplaner BlendedPanningNormals, parameterName=BPN_Weapon_Offset, sortPriority=11, defaultValue=<rgba 0 0 0 1>}
  n11 = VectorParameter() {group=Primary Triplaner BlendedPanningNormals, parameterName=BPN_Vehicle_Offset, sortPriority=12, defaultValue=<rgba 0 0 0 1>}
  n12 = VectorParameter() {group=Primary Triplaner BlendedPanningNormals, parameterName=BPN_Mech_Offset, sortPriority=10, defaultValue=<rgba 0 0 0 1>}
  n13 = LinearInterpolate(A=LinearInterpolate(A=@n10.RGB, B=@n11.RGB, Alpha=@n7), B=@n12.RGB, Alpha=@n8)
  n14 = PreSkinnedLocalBounds()
  n15 = Subtract(A=@n14.Max, B=@n14.Max)
  n16 = VectorParameter() {group=Primary Triplaner BlendedPanningNormals, parameterName=BPN_Panning_Weapon_Speed, sortPriority=21, defaultValue=<rgba 0 0 0 1>}
  n17 = VectorParameter() {group=Primary Triplaner BlendedPanningNormals, parameterName=BPN_Panning_Vehicle_Speed, sortPriority=22, defaultValue=<rgba 0 0 0 1>}
  n18 = VectorParameter() {group=Primary Triplaner BlendedPanningNormals, parameterName=BPN_Panning_Mech_Speed, sortPriority=20, defaultValue=<rgba 0 0 0 1>}
  n19 = LinearInterpolate(A=LinearInterpolate(A=@n16.RGB, B=@n17.RGB, Alpha=@n7), B=@n18.RGB, Alpha=@n8)
  n20 = Add(
   A=Add(A=FunctionInput() {bUsePreviewValueAsDefault=true, inputName=Pre-Skinned Local Position, sortPriority=10}, B=Multiply(A=@n13, B=@n15))
   B=Multiply(A=Time(), B=Multiply(A=@n19, B=@n15))
  )
  n21 = TextureCoordinate()
  n22 = ScalarParameter() {defaultValue=1, group=Primary Triplaner BlendedPanningNormals, parameterName=BPN_WarpUvTileRate, sortPriority=52}
  n23 = Time() {period=1}
  n24 = AppendVector(
   A=ScalarParameter() {group=Primary Triplaner BlendedPanningNormals, parameterName=BPN_WarpUvSpeedX, sortPriority=58}
   B=ScalarParameter() {group=Primary Triplaner BlendedPanningNormals, parameterName=BPN_WarpUvSpeedY, sortPriority=59}
  )
  n25 = TextureSampleParameter2D(UVs=Add(A=Multiply(A=@n21, B=@n22), B=Multiply(A=@n23, B=@n24))) {
   group=Primary Triplaner BlendedPanningNormals
   parameterName=BPN_WarpUvTexture
   sortPriority=51
   texture=&AtmosphericCloudNoise01
  }
  n26 = ScalarParameter() {group=Primary Triplaner BlendedPanningNormals, parameterName=BPN_WarpUvAmount Weapon, sortPriority=55}
  n27 = ScalarParameter() {group=Primary Triplaner BlendedPanningNormals, parameterName=BPN_WarpUvAmount_Vehicle, sortPriority=56}
  n28 = Multiply(
   A=ChannelMaskParameter(None=@n25.RGB) {group=Primary Triplaner BlendedPanningNormals, parameterName=BPN_WarpUVsChannel, sortPriority=51}.R
   B=LinearInterpolate(
    A=LinearInterpolate(A=@n26, B=@n27, Alpha=@n7)
    B=ScalarParameter() {group=Primary Triplaner BlendedPanningNormals, parameterName=BPN_WarpUvAmount_Mech, sortPriority=57}
    Alpha=@n8
   )
  )
  n29 = MaterialFunctionCall!(
   Material Normals=@n3
   TextureObject_MASK
   TextureObject_Diffuse
   TextureObject_Normal=@n4
   TextureScale=LinearInterpolate(
    A=LinearInterpolate(A=@n5, B=@n6, Alpha=@n7)
    B=ScalarParameter() {defaultValue=60, group=Primary Triplaner BlendedPanningNormals, parameterName=BPN_Mech_Texture_Scale, sortPriority=4}
    Alpha=@n8
   )
   Normal Texture Intensity=@n9
   Pre-Skinned Local Position=StaticSwitchParameter(True=Add(A=@n20, B=@n28), False=@n20) {group=Primary Triplaner BlendedPanningNormals, parameterName=BPN_Warp_UVs, sortPriority=50}
   Pre-Skinned Local Normal=FunctionInput() {bUsePreviewValueAsDefault=true, inputName=Pre-Skinned Local Normal, sortPriority=11}
   Rotation Z Axis=Add(
    A=FunctionInput() {bUsePreviewValueAsDefault=true, inputName=Rotation Z Axis, inputType=FunctionInput_Scalar, sortPriority=1}
    B=ScalarParameter() {group=Primary Triplaner, parameterName=Global X Rot, sortPriority=-40}
   )
   Rotation X Axis=Add(
    A=FunctionInput() {bUsePreviewValueAsDefault=true, inputName=Rotation X Axis, inputType=FunctionInput_Scalar, sortPriority=2}
    B=ScalarParameter() {group=Primary Triplaner, parameterName=Global Z Rot, sortPriority=-35}
   )
   Triplanner_Offset
   isMech=@n8
   Export Float 4
   ProjectionTransitionContrast=ScalarParameter() {defaultValue=1, group=Primary Triplaner BlendedPanningNormals, parameterName=BPN_Triplaners Seam Hardness, sortPriority=6}
   IsStatsB
   isCharacter
  ) {materialFunction=/Game/Packages/Fortress_SharedMaterials/Base_Material_Functions/BaseCustomizationFunctions/MF_Base_Customization_Triplanar}
  n30 = FunctionInput() {inputName=Tirplaner01Mask, inputType=FunctionInput_Vector4, sortPriority=3}
  n31 = ChannelMaskParameter(None=@n30) {desc=Overrided To TopSideProjection, group=Primary Triplaner BlendedPanningNormals, parameterName=BPN_TS_Mask01_Channel, sortPriority=1}
  n32 = ChannelMaskParameter(None=@n30) {group=Primary Triplaner BlendedPanningNormals, parameterName=BPN_Tirplaner01Mask Channel, sortPriority=1}
  n33 = StaticSwitchParameter(
   True=LinearInterpolate(A=@n3, B=@n29.Normal Texture, Alpha=StaticSwitchParameter(True=@n31.R, False=@n32.R) {group=Customization, parameterName=TopSideProjection})
   False=@n29.Normal Texture
  ) {group=Primary Triplaner BlendedPanningNormals, parameterName="BPN_Mask Normals Tirplaner01Mask "}
  n34 = FunctionInput() {bUsePreviewValueAsDefault=true, inputName="Customization Mask (_CM B)", inputType=FunctionInput_Scalar, sortPriority=30}
  n35 = SetMaterialAttributes!(
   MaterialAttributes=@n2.MaterialAttributes
   Normal=LinearInterpolate(
    A=@n2.Normal
    B=MaterialFunctionCall(BaseNormal=@n2.Normal, AdditionalNormal=@n33) {materialFunction=&BlendAngleCorrectedNormals}.Result
    Alpha=@n34
   )
  ) {attributeSetTypes [#1]}
  BlendMaterialAttributes(A=@n1, B=@n35, Alpha=@n34)
  FunctionOutput(None=StaticSwitchParameter(True=@n35, False=@n1) {group=Customization, parameterName=Triplaners Blended Panning Normals})
 ]}
