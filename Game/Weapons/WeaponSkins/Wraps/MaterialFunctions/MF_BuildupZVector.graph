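# Wrap material function that layers a buildup material, leaves by default, onto weapons, vehicles and mechs, behind the BuildUpZVector switch.
# Local-normal Z, Gmap and bounds gradients drive height lerps against the first texture's alpha, shaping the blend mask with the customization mask.
# Bounds-normalized positions feed triplanar projection, with per-item tiling.
# Two optional alpha-masked layers scroll over time using triplanar or top-side projection, then everything blends into the incoming material attributes.
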MaterialFunction /Game/Weapons/WeaponSkins/Wraps/MaterialFunctions/MF_BuildupZVector
graph {nodes [
  n1 = GetMaterialAttributes(None=FunctionInput() {inputName=Material, inputType=FunctionInput_MaterialAttributes, sortPriority=-100}) {attributeGetTypes [#1, #2]}
  n2 = TextureObjectParameter() {
   desc=Alpha is mask
   group=Primary_BuildUpZ
   parameterName=BuildUp_Diffuse02
   sortPriority=11
   texture=/Game/Weapons/WeaponSkins/Wraps/Textures/Leaves/T_Wrap_leaves_BC2
  }
  n3 = TextureObjectParameter() {
   desc=Alpha is mask
   group=Primary_BuildUpZ
   parameterName=BuildUp_Diffuse03
   sortPriority=22
   texture=&T_Wrap_leaves_BC2
  }
  n4 = StaticBoolParameter() {group=Primary_BuildUpZ, parameterName="Use Textures02 For Textures03 ", sortPriority=21}
  n5 = StaticSwitch(True=@n2, False=@n3, Value=@n4)
  n6 = FunctionInput(Preview=Constant()) {bUsePreviewValueAsDefault=true, inputName=isVehicle, inputType=FunctionInput_Scalar, sortPriority=40}
  n7 = FunctionInput(Preview=Constant()) {bUsePreviewValueAsDefault=true, inputName=isMech, inputType=FunctionInput_Scalar, sortPriority=41}
  n8 = LinearInterpolate(
   A=LinearInterpolate(A=Constant3Vector() {constant=<rgba 0.5 0.5 0.25 1>}, B=Constant3Vector() {constant=<rgba 0.5 0.5 0.5 1>}, Alpha=@n6)
   B=Constant3Vector() {constant=<rgba -0.25 -0.25 0 1>}
   Alpha=@n7
  )
  n9 = ScalarParameter() {defaultValue=1, group=Primary_BuildUpZ, parameterName=BuidUp_Alternate_Axis_Opacity, sortPriority=11}
  n10 = VectorParameter() {group=Primary_BuildUpZ, parameterName=BuidUp_Alternate_Axis_Color03, sortPriority=22}
  n11 = FunctionInput(Preview=VertexInterpolator(VS=PreSkinnedNormal()).PS) {bUsePreviewValueAsDefault=true, inputName=Local Normals}
  n12 = ScalarParameter() {defaultValue=1, group=Primary_BuildUpZ, parameterName=BuidUp_TriPlanner_Contrast, sortPriority=10}
  n13 = PreSkinnedLocalBounds()
  n14 = MaterialFunctionCall(Color=@n13.Min) {materialFunction=&SplitComponents}
  n15 = AppendVector(A=@n14.R, B=Constant())
  n16 = LinearInterpolate(A=LinearInterpolate(A=@n13.Min, B=AppendVector(A=@n15, B=@n14.B), Alpha=@n6), B=@n13.Min, Alpha=@n7)
  n17 = FunctionInput(Preview=VertexInterpolator(VS=PreSkinnedPosition()).PS) {bUsePreviewValueAsDefault=true, inputName=Local Position}
  n18 = MaterialFunctionCall(Color=Subtract(A=@n13.Max, B=@n13.Max)) {materialFunction=&SplitComponents}
  n19 = Max(A=@n18.R, B=@n18.R)
  n20 = LinearInterpolate(
   A=ScalarParameter() {defaultValue=2, group=Primary_BuildUpZ, parameterName=BuildUp_TileRate_Weapon, sortPriority=45}
   B=ScalarParameter() {defaultValue=5, group=Primary_BuildUpZ, parameterName=BuildUp_TileRate_Vehicle, sortPriority=46}
   Alpha=@n6
  )
  n21 = LinearInterpolate(
   A=ScalarParameter() {defaultValue=5, group=Primary_BuildUpZ, parameterName=BuildUp_TileRate_Mech, sortPriority=47}
   B=ScalarParameter() {defaultValue=1, group=Primary_BuildUpZ, parameterName=BuildUp_TileRate_Backpack, sortPriority=47}
   Alpha=NamedRerouteUsage()
  )
  n22 = Multiply(
   A=Divide(A=Subtract(A=@n17, B=@n16), B=Max(A=@n19, B=@n18.B))
   B=LinearInterpolate(
    A=LinearInterpolate(A=@n20, B=@n21, Alpha=@n7)
    B=ScalarParameter() {defaultValue=1, group=Primary_BuildUpZ, parameterName=BuildUp_TileRate_Pickaxe, sortPriority=47}
    Alpha=NamedRerouteUsage()
   )
  )
  n23 = Time()
  n24 = VectorParameter() {group=Primary_BuildUpZ, parameterName=BuildUp_03_Speed_Weapon, sortPriority=60, defaultValue=<rgba 0 0.75 0 1>}
  n25 = VectorParameter() {group=Primary_BuildUpZ, parameterName=BuildUp_03_Speed_Vehicle, sortPriority=61, defaultValue=<rgba 1 0 0 1>}
  n26 = VectorParameter() {group=Primary_BuildUpZ, parameterName=BuildUp_03_Speed_Mech, sortPriority=62, defaultValue=<rgba 0 0.75 0 1>}
  n27 = LinearInterpolate(A=LinearInterpolate(A=@n24.RGB, B=@n25.RGB, Alpha=@n6), B=@n26.RGB, Alpha=@n7)
  n28 = StaticBool() {value=true}
  n29 = StaticSwitch(True=@n27, False=ComponentMask(None=@n27) {g=true, r=true}, Value=@n28)
  n30 = Constant()
  n31 = Time()
  n32 = ScalarParameter() {defaultValue=0.05, group=Primary_BuildUpZ, parameterName=BuildUp_03_Sine_Speed, sortPriority=63}
  n33 = LinearInterpolate(
   A=ScalarParameter() {defaultValue=1, group=Primary_BuildUpZ, parameterName=BuildUp_03_Sine_Min, sortPriority=64}
   B=ScalarParameter() {defaultValue=-1, group=Primary_BuildUpZ, parameterName=BuildUp_03_Sine_Max, sortPriority=65}
   Alpha=Sine(None=Multiply(A=@n31, B=@n32))
  )
  n34 = Constant()
  n35 = AppendVector(A=@n34, B=@n33)
  n36 = LinearInterpolate(A=LinearInterpolate(A=AppendVector(A=@n33, B=@n34), B=@n35, Alpha=@n6), B=@n35, Alpha=@n7)
  n37 = Add(A=@n22, B=Add(A=Multiply(A=@n23, B=@n29), B=StaticSwitch(True=@n30, False=@n36, Value=@n28)))
  n38 = MaterialFunctionCall!(
   IsCharacter
   IsStatsB
   Texture=@n5
   Rotation_Axis=@n8
   Caculated_Bounds=@n37
   Pre_Skinned_Normals=@n11
   Blend_Contrast=@n12
   Fill_alternate_Axis_Opacity=@n9
   "Fill_alternate_Axis_Color_A=Roughness"=AppendVector(A=@n10.RGB, B=@n10.RGB)
   Is_Vehicle=@n6
   Is_Mech=@n7
   Enable_Fill_Z_Weapon
   Fill_Z_Weapon_opacity
   Fill_Z_Weapon_Color
  ) {materialFunction=&/MF_TopsSideProjection_Base}
  n39 = TextureObjectParameter() {
   group=Primary_BuildUpZ
   parameterName=BuildUP_Normals02
   samplerType=SAMPLERTYPE_Normal
   sortPriority=12
   texture=&FlatNormalBlue
  }
  n40 = TextureObjectParameter() {
   group=Primary_BuildUpZ
   parameterName=BuildUP_Normals03
   samplerType=SAMPLERTYPE_Normal
   sortPriority=23
   texture=&FlatNormalBlue
  }
  n41 = StaticSwitch(True=@n39, False=@n40, Value=@n4)
  n42 = Constant() {r=1}
  n43 = StaticBool() {value=true}
  n44 = Add(
   A=FunctionInput() {bUsePreviewValueAsDefault=true, inputName=Rotation X Axis, inputType=FunctionInput_Scalar}
   B=ScalarParameter() {group=Primary Triplaner, parameterName=Global Z Rot, sortPriority=-35}
  )
  n45 = Add(
   A=FunctionInput() {bUsePreviewValueAsDefault=true, inputName=Rotation Z Axis, inputType=FunctionInput_Scalar}
   B=ScalarParameter() {group=Primary Triplaner, parameterName=Global X Rot, sortPriority=-40}
  )
  n46 = ScalarParameter() {defaultValue=1, group=Primary_BuildUpZ, parameterName=BuidUp_TriPlanner_Normal_Strenght, sortPriority=10}
  n47 = Constant3Vector() {constant=<rgba 0 0 1 0>}
  n48 = MaterialFunctionCall!(
   Material Normals=@n47
   TextureObject_MASK
   TextureObject_Diffuse=@n5
   TextureObject_Normal=@n41
   TextureScale=@n42
   Normal Texture Intensity=@n46
   Pre-Skinned Local Position=@n37
   Pre-Skinned Local Normal=@n11
   Rotation Z Axis=@n45
   Rotation X Axis=@n44
   Triplanner_Offset
   isMech=@n7
   Export Float 4=@n43
   ProjectionTransitionContrast=@n12
   IsStatsB
   isCharacter
  ) {materialFunction=/Game/Packages/Fortress_SharedMaterials/Base_Material_Functions/BaseCustomizationFunctions/MF_Base_Customization_Triplanar}
  n49 = StaticBoolParameter() {desc=Rotate texture 180., group=Primary_BuildUpZ, parameterName=Use_TopSideProjection Textures02 and 03, sortPriority=7}
  n50 = StaticSwitch(True=@n38.RGB, False=@n48.Diffuse Texture, Value=@n49)
  n51 = StaticSwitch(True=@n38.Alpha, False=@n48.Roughness Texture, Value=@n49)
  n52 = VectorParameter() {group=Primary_BuildUpZ, parameterName=BuidUp_Alternate_Axis_Color02, sortPriority=11}
  n53 = Time()
  n54 = VectorParameter() {group=Primary_BuildUpZ, parameterName=BuildUp_02_Speed_Weapon, sortPriority=50, defaultValue=<rgba 0 1 0 1>}
  n55 = VectorParameter() {group=Primary_BuildUpZ, parameterName=BuildUp_02_Speed_Vehicle, sortPriority=51, defaultValue=<rgba 1.5 0 0 1>}
  n56 = VectorParameter() {group=Primary_BuildUpZ, parameterName=BuildUp_02_Speed_Mech, sortPriority=52, defaultValue=<rgba 0 1 0 1>}
  n57 = LinearInterpolate(A=LinearInterpolate(A=@n54.RGB, B=@n55.RGB, Alpha=@n6), B=@n56.RGB, Alpha=@n7)
  n58 = StaticSwitch(True=@n57, False=ComponentMask(None=@n57) {g=true, r=true}, Value=@n28)
  n59 = Constant()
  n60 = Time()
  n61 = ScalarParameter() {defaultValue=0.05, group=Primary_BuildUpZ, parameterName=BuildUp_02_Sine_Speed, sortPriority=53}
  n62 = LinearInterpolate(
   A=ScalarParameter() {defaultValue=-1, group=Primary_BuildUpZ, parameterName=BuildUp_02_Sine_Min, sortPriority=54}
   B=ScalarParameter() {defaultValue=1, group=Primary_BuildUpZ, parameterName=BuildUp_02_Sine_Max, sortPriority=55}
   Alpha=Sine(None=Multiply(A=@n60, B=@n61))
  )
  n63 = Constant()
  n64 = AppendVector(A=@n63, B=@n62)
  n65 = LinearInterpolate(A=LinearInterpolate(A=AppendVector(A=@n62, B=@n63), B=@n64, Alpha=@n6), B=@n64, Alpha=@n7)
  n66 = Add(A=@n22, B=Add(A=Multiply(A=@n53, B=@n58), B=StaticSwitch(True=@n59, False=@n65, Value=@n28)))
  n67 = MaterialFunctionCall(
   Texture=@n2
   Rotation_Axis=@n8
   Caculated_Bounds=@n66
   Pre_Skinned_Normals=@n11
   Blend_Contrast=@n12
   Fill_alternate_Axis_Opacity=@n9
   "Fill_alternate_Axis_Color_A=Roughness"=AppendVector(A=@n52.RGB, B=@n52.RGB)
   Is_Vehicle=@n6
   Is_Mech=@n7
  ) {materialFunction=&MF_TopsSideProjection_Base}
  n68 = MaterialFunctionCall(
   Material Normals=@n47
   TextureObject_Diffuse=@n2
   TextureObject_Normal=@n39
   TextureScale=@n42
   Normal Texture Intensity=@n46
   Pre-Skinned Local Position=@n66
   Pre-Skinned Local Normal=@n11
   Rotation Z Axis=@n45
   Rotation X Axis=@n44
   isMech=@n7
   Export Float 4=@n43
   ProjectionTransitionContrast=@n12
  ) {materialFunction=&MF_Base_Customization_Triplanar}
  n69 = StaticSwitch(True=@n67.RGB, False=@n68.Diffuse Texture, Value=@n49)
  n70 = LinearInterpolate(A=@n69, B=@n50, Alpha=@n51)
  n71 = StaticBoolParameter() {group=Primary_BuildUpZ, parameterName=Add_Masked_Diffuse/Normal03, sortPriority=20}
  n72 = StaticSwitch(True=@n67.Alpha, False=@n68.Roughness Texture, Value=@n49)
  n73 = TextureObjectParameter() {
   group=Primary_BuildUpZ
   parameterName=BuildUp_Mask01
   samplerType=SAMPLERTYPE_Masks
   sortPriority=8
   texture=&T_BlackMask
  }
  n74 = TextureObjectParameter() {
   desc=Alpha is heightmap
   group=Primary_BuildUpZ
   parameterName=BuildUp_Diffuse01
   sortPriority=6
   texture=/Game/Weapons/WeaponSkins/Wraps/Textures/Leaves/T_Wrap_leaves_BC
  }
  n75 = TextureObjectParameter() {
   group=Primary_BuildUpZ
   parameterName=BuildUP_Normals01
   samplerType=SAMPLERTYPE_Normal
   sortPriority=7
   texture=&FlatNormalBlue
  }
  n76 = MaterialFunctionCall(
   Material Normals=@n47
   TextureObject_MASK=@n73
   TextureObject_Diffuse=@n74
   TextureObject_Normal=@n75
   TextureScale=@n42
   Normal Texture Intensity=@n46
   Pre-Skinned Local Position=@n22
   Pre-Skinned Local Normal=@n11
   Rotation Z Axis=@n45
   Rotation X Axis=@n44
   isMech=@n7
   Export Float 4=@n43
   ProjectionTransitionContrast=@n12
  ) {materialFunction=&MF_Base_Customization_Triplanar}
  n77 = MaterialFunctionCall!(
   is Vehicle=@n6
   Object Bounds Max Adjustment=FunctionInput() {bUsePreviewValueAsDefault=true, inputName=Object Bounds Max Adjustment, sortPriority=52}
   Object Bounds Min Adjustment=FunctionInput() {bUsePreviewValueAsDefault=true, inputName=Object Bounds Min Adjustment, sortPriority=51}
   Object Bounds Center Adjustment=FunctionInput() {bUsePreviewValueAsDefault=true, inputName=Object Bounds Center Adjustment, sortPriority=50}
   isMech=@n7
   Side Project Vehicle
   Local Position=@n17
  ) {materialFunction=&/MF_WrapBounds}
  n78 = ScalarParameter() {defaultValue=-0.3, group=Primary_BuildUpZ, parameterName=BuldUp_Grad_offset_Weapon Z, sortPriority=39}
  n79 = ComponentMask(None=Divide(
    A=Add(A=@n77.Result, B=@n78)
    B=ScalarParameter() {defaultValue=0.25, group=Primary_BuildUpZ, parameterName=BuldUp_Grad_Falloff_Weapon Z, sortPriority=39}
   )) {b=true}
  n80 = StaticSwitch(
   True=@n79
   False=OneMinus(None=@n79)
   Value=StaticBoolParameter() {defaultValue=true, group=Primary_BuildUpZ, parameterName=BuldUp_Grad_Invert_Weapon Z, sortPriority=39}
  )
  n81 = ComponentMask(None=@n11) {b=true}
  n82 = Abs(None=ComponentMask(None=@n11) {r=true})
  n83 = Add(A=Saturate(None=@n81), B=@n82)
  n84 = ScalarParameter() {defaultValue=1.75, group=Primary_BuildUpZ, parameterName=BuldUp_Grad_X_Mask_Falloff_Weapon, sortPriority=39}
  n85 = ScalarParameter() {defaultValue=5, group=Primary_BuildUpZ, parameterName=BuldUp_Grad_X_Mask_Contrast_Weapon, sortPriority=39}
  n86 = Saturate(None=MaterialFunctionCall(In=Divide(A=@n83, B=@n84), Contrast=@n85) {materialFunction=&CheapContrast}.Result)
  n87 = LinearInterpolate(A=Saturate(None=Multiply(A=@n80, B=@n86)), Alpha=@n6) {constB=0}
  n88 = ScalarParameter() {defaultValue=2, group=Primary_BuildUpZ, parameterName=BuildUp_Z_Strenght_Weapon}
  n89 = Subtract(
   A=FunctionInput(Preview=TextureSample() {samplerType=SAMPLERTYPE_Masks, texture=/Game/Weapons/FORT_RocketLaunchers/Textures/T_Military_M202_Launcher_CM}.A) {bUsePreviewValueAsDefault=true, inputName=Gmap, inputType=FunctionInput_Scalar, sortPriority=24}
   B=ScalarParameter() {defaultValue=0.75, group=Primary_BuildUpZ, parameterName=BuldUp_Gmap_influence Z, sortPriority=39}
  )
  n90 = ScalarParameter() {defaultValue=10, parameterName=BuldUp_Gmap_Strength Z}
  n91 = Multiply(A=ComponentMask(None=@n11) {b=true}, B=ScalarParameter() {defaultValue=8, parameterName=BuldUp_LNormal_Strength Z})
  n92 = Saturate(None=Add(A=Multiply(A=@n89, B=@n90), B=Saturate(None=@n91)))
  n93 = Saturate(None=LinearInterpolate(A=Multiply(A=@n92, B=@n88), B=@n92, Alpha=@n6))
  n94 = MaterialFunctionCall(
   Transition Phase=Saturate(None=Add(A=@n87, B=@n93))
   Height Texture=@n76.Roughness Texture
   Contrast=ScalarParameter() {defaultValue=5, group=Primary_BuildUpZ, parameterName=BuildUp_Z_Contrast}
  ) {materialFunction=&HeightLerp}
  n95 = MaterialFunctionCall(
   Transition Phase=MaterialFunctionCall!(
    ForwardAxis=FunctionInput() {bUsePreviewValueAsDefault=true, inputName=Forward Axis}
    Bounds=@n77.Result
    Pre-Skinned Local Position
    isVehicle=@n6
    isMech=@n7
    Side Project Vehicle
    WeaponOffset=ScalarParameter() {group=Primary_BuildUpZ, parameterName=BuldUp_Grad_offset_Weapon, sortPriority=40}
    WeaponFalloff=ScalarParameter() {defaultValue=0.8, group=Primary_BuildUpZ, parameterName=BuldUp_Grad_Falloff_Weapon, sortPriority=41}
    VehicleOffset=ScalarParameter() {group=Primary_BuildUpZ, parameterName=BuldUp_Grad_offset_Vehicle, sortPriority=42}
    VehicleFalloff=ScalarParameter() {defaultValue=1, group=Primary_BuildUpZ, parameterName=BuldUp_Grad_Falloff_Vehicle, sortPriority=43}
    MechOffset=ScalarParameter() {group=Primary_BuildUpZ, parameterName=BuldUp_Grad_offset_Mech, sortPriority=44}
    MechFalloff=ScalarParameter() {defaultValue=1, group=Primary_BuildUpZ, parameterName=BuldUp_Grad_Falloff_Mech, sortPriority=45}
    Invert_Gradent=StaticBoolParameter() {defaultValue=true, group=Primary_BuildUpZ, parameterName=BuildUp_Gradent_invert--------------------, sortPriority=40}
    UV_Spread_XY
    UV_Animated_XY
    Warp_UVs
    Rotate_Mech
    Rotate_Vehicle_Side
    Rotate_Vehicle
   ) {materialFunction=&/MF_Wrap_Gradent_Creation}.Gradent
   Height Texture=@n76.Roughness Texture
   Contrast=ScalarParameter() {defaultValue=5, group=Primary_BuildUpZ, parameterName=BuildUp_Grad_HeightLerp_Contrast, sortPriority=38}
  ) {materialFunction=&HeightLerp}
  n96 = FunctionInput(Preview=Constant() {r=1}) {bUsePreviewValueAsDefault=true, inputName="Customization Mask (_CM B Top left)", inputType=FunctionInput_Scalar, sortPriority=24}
  n97 = ScalarParameter() {group=Primary_BuildUpZ, parameterName=Buildup_SecondaryArea_Weapon, sortPriority=-4}
  n98 = ScalarParameter() {group=Primary_BuildUpZ, parameterName=Buildup_SecondaryArea_Vehicle, sortPriority=-4}
  n99 = Saturate(None=Add(A=@n96, B=LinearInterpolate(A=@n97, B=@n98, Alpha=@n6) {constB=0}))
  n100 = LinearInterpolate(A=OneMinus(None=@n72), B=Multiply(A=@n94.Alpha, B=@n95.Alpha), Alpha=@n99)
  n101 = LinearInterpolate(
   A=@n51
   B=Multiply(A=@n51, B=OneMinus(None=@n100))
   Alpha=ScalarParameter() {group=Primary_BuildUpZ, parameterName=Diffuse03_Draw_Bhind_Diffuse01, sortPriority=21}
  )
  n102 = StaticBoolParameter() {group=Primary_BuildUpZ, parameterName=Add_Masked_Diffuse/Normal02, sortPriority=10}
  n103 = StaticSwitch(True=LinearInterpolate(A=@n69, B=@n76.Diffuse Texture, Alpha=@n100), False=@n76.Diffuse Texture, Value=@n102)
  n104 = LinearInterpolate(A=@n103, B=@n50, Alpha=@n101)
  n105 = StaticBoolParameter() {group=Primary_BuildUpZ, parameterName=Disable_MainTextures, sortPriority=7}
  n106 = StaticSwitch(
   True=Constant3Vector() {constant=<rgba 0 0 1 0>}
   False=StaticSwitch(
    True=MaterialFunctionCall!(
     Texture_Normals=@n39
     Rotation_Axis=@n8
     Caculated_Bounds=@n66
     Pre_Skinned_Normals=@n11
     Blend_Contrast=@n12
     Fill_alternate_Axis_Opacity=@n9
     Is_Vehicle=@n6
     Is_Mech=@n7
    ) {materialFunction=&/MF_TopsSideProjection_Normals}.Add_Surface_Normal
    False=@n68.Normal Texture
    Value=@n49
   )
   Value=StaticBoolParameter() {group=Primary_BuildUpZ, parameterName=Disable_normals02, sortPriority=12}
  )
  n107 = StaticSwitch(
   True=Constant3Vector() {constant=<rgba 0 0 1 0>}
   False=StaticSwitch(
    True=MaterialFunctionCall(
     Texture_Normals=@n41
     Rotation_Axis=@n8
     Caculated_Bounds=@n37
     Pre_Skinned_Normals=@n11
     Blend_Contrast=@n12
     Fill_alternate_Axis_Opacity=@n9
     Is_Vehicle=@n6
     Is_Mech=@n7
    ) {materialFunction=&MF_TopsSideProjection_Normals}.Add_Surface_Normal
    False=@n48.Normal Texture
    Value=@n49
   )
   Value=StaticBoolParameter() {group=Primary_BuildUpZ, parameterName=Disable_normals03, sortPriority=22}
  )
  n108 = LinearInterpolate(A=@n106, B=StaticSwitch(True=@n106, False=@n107, Value=@n4), Alpha=@n51)
  n109 = StaticSwitch(
   True=Constant3Vector() {constant=<rgba 0 0 1 0>}
   False=@n76.Normal Texture
   Value=StaticBoolParameter() {group=Primary_BuildUpZ, parameterName=Disable_normals01, sortPriority=7}
  )
  n110 = StaticSwitch(True=LinearInterpolate(A=@n106, B=@n109, Alpha=@n100), False=@n109, Value=@n102)
  n111 = LinearInterpolate(A=@n110, B=@n107, Alpha=@n101)
  n112 = ScalarParameter() {group=Primary_BuildUpZ, parameterName=BuildUp_Roughness_Min, sortPriority=5}
  n113 = ScalarParameter() {defaultValue=1, group=Primary_BuildUpZ, parameterName=BuildUp_Roughness_Max, sortPriority=5}
  n114 = ChannelMaskParameter(None=@n76.Output MASK) {group=Primary_BuildUpZ, parameterName=BuildUp_Roughness_Channel, sortPriority=5, defaultValue=<rgba 0 1 0 0>}
  n115 = ChannelMaskParameter(None=@n76.Output MASK) {group=Primary_BuildUpZ, maskChannel=Green, parameterName=BuildUp_Metallic_Channel, sortPriority=5, defaultValue=<rgba 0 1 0 0>}
  n116 = SetMaterialAttributes!(
   MaterialAttributes=@n1.MaterialAttributes
   BaseColor=StaticSwitch(True=StaticSwitch(True=@n70, False=@n69, Value=@n71), False=StaticSwitch(True=@n104, False=@n103, Value=@n71), Value=@n105)
   Normal=StaticSwitch(True=StaticSwitch(True=@n108, False=@n106, Value=@n71), False=StaticSwitch(True=@n111, False=@n110, Value=@n71), Value=@n105)
   Roughness=StaticSwitchParameter(
    True=LinearInterpolate(A=@n112, B=@n113, Alpha=@n114.R)
    False=ScalarParameter() {defaultValue=0.5, group=Primary_BuildUpZ, parameterName=BuildUp_Roughness, sortPriority=5}
   ) {group=Primary_BuildUpZ, parameterName=BuildUp_Roughness_Texture, sortPriority=5}
   Metallic=StaticSwitchParameter(True=@n115.R, False=ScalarParameter() {group=Primary_BuildUpZ, parameterName=BuildUp_Metallic, sortPriority=5}) {group=Primary_BuildUpZ, parameterName=BuildUp_Metallic_Texture, sortPriority=5}
   Specular=ScalarParameter() {defaultValue=0.5, group=Primary_BuildUpZ, parameterName=BuildUp_Specular, sortPriority=5}
  ) {attributeSetTypes [
    #2
    #1
    #3
    #4
    #5
   ]}
  n117 = ScalarParameter() {defaultValue=1, group=Primary_BuildUpZ, parameterName=Buildup_MaskSecondary Material, sortPriority=-5}
  n118 = Saturate(None=Add(A=@n72, B=@n51))
  n119 = StaticSwitch(True=Add(A=@n72, B=@n51), False=@n72, Value=@n71)
  n120 = Saturate(None=Multiply(A=@n94.Alpha, B=@n95.Alpha))
  n121 = Saturate(None=@n94.Alpha)
  n122 = Add(A=@n120, B=Multiply(A=@n121, B=@n72))
  n123 = Multiply(A=@n121, B=@n51)
  n124 = StaticSwitch(True=StaticSwitch(True=Add(A=@n122, B=@n123), False=@n122, Value=@n71), False=@n120, Value=@n102)
  n125 = Add(A=StaticSwitch(True=Saturate(None=@n119), False=Constant(), Value=@n102), B=Multiply(A=Saturate(None=@n124), B=@n99))
  n126 = Multiply(
   A=Saturate(None=Add(A=@n117, B=@n96))
   B=StaticSwitch(True=StaticSwitch(True=@n118, False=@n72, Value=@n71), False=Saturate(None=@n125), Value=@n105)
  )
  FunctionOutput(None=StaticSwitchParameter(True=BlendMaterialAttributes(A=@n1.MaterialAttributes, B=@n116, Alpha=@n126) {vertexAttributeBlendType=UseA}, False=@n1.MaterialAttributes) {group=Customization, parameterName=BuildUpZVector})
  n127 = NamedRerouteDeclaration(None=FunctionInput() {bUsePreviewValueAsDefault=true, inputName=IsStatsB, inputType=FunctionInput_Vector2, sortPriority=78}) {name=Is_StatsB, nodeColor=<rgba 0 0.9253 1 1>}
  NamedRerouteDeclaration(None=ComponentMask(None=@n127) {r=true}) {name=IsBackpack, nodeColor=<rgba 0.2656 0 1 1>}
  NamedRerouteDeclaration(None=ComponentMask(None=@n127) {g=true}) {name=IsPickaxe, nodeColor=<rgba 1 0 0.5625 1>}
  Normalize(VectorInput=Transform(None=@n1.Normal) {transformType=TRANSFORM_Local})
  n128 = MaterialFunctionCall(In=Abs(None=@n81), Contrast=@n85) {materialFunction=&CheapContrast}
  n129 = MaterialFunctionCall(In=@n82, Contrast=@n85) {materialFunction=&CheapContrast}
  Saturate(None=Add(A=@n128.Result, B=@n129.Result))
  n130 = MaterialFunctionCall(In=Abs(None=ComponentMask(None=@n11) {g=true}), Contrast=@n85) {materialFunction=&CheapContrast}
  Saturate(None=Add(A=@n128.Result, B=@n130.Result))
  Saturate(None=Add(A=@n129.Result, B=@n130.Result))
  StaticBoolParameter() {defaultValue=true, group=Primary_BuildUpZ, parameterName=BuidUp_TriPlanner_Textures02 and 03, sortPriority=10}
  TextureObjectParameter() {
   desc=Alpha is mask
   group=Primary_BuildUpZ
   parameterName=BuildUp_Diffuse03
   sortPriority=22
   texture=&T_BlacksRGB
  }
 ]}
