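# Carnaval wrap skin material function that layers an animated look over the incoming material attributes, with variants for weapons, vehicles and mechs.
# Wrap bounds drive a scrolling background, mask-warped strips and two projected star layers that height-blend into hue-shifted emissive.
# Time-stamp parameters trigger star motion and fades, plus a gradient-colored horn wipe that covers the other layers.
# The Carnaval static switch enables the result, blended by the customization mask.
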MaterialFunction /Game/Weapons/WeaponSkins/Wraps/MaterialFunctions/MF_Carnaval
graph {nodes [
  n1 = FunctionInput() {inputType=FunctionInput_MaterialAttributes}
  n2 = GetMaterialAttributes(None=@n1) {attributeGetTypes [#1, #2]}
  n3 = ScalarParameter() {defaultValue=-3, group=Wrap_Carnaval, parameterName=CarnB_StarsB_Offset}
  n4 = ScalarParameter() {defaultValue=1.75, group=Wrap_Carnaval, parameterName=CarnB_StarsB_Falloff}
  n5 = FunctionInput(Preview=Constant3Vector() {constant=<rgba 0 1 0 0>}) {bUsePreviewValueAsDefault=true, inputName=ForwardAxis}
  n6 = FunctionInput(Preview=Constant()) {bUsePreviewValueAsDefault=true, inputName=isVehicle, inputType=FunctionInput_Scalar, sortPriority=4}
  n7 = FunctionInput(Preview=Constant()) {bUsePreviewValueAsDefault=true, inputName=isMech, inputType=FunctionInput_Scalar}
  n8 = PreSkinnedPosition()
  n9 = MaterialFunctionCall!(
   is Vehicle=@n6
   Object Bounds Max Adjustment=FunctionInput(Preview=Constant3Vector() {constant=<rgba -5 -25 -5 0>}) {bUsePreviewValueAsDefault=true, inputName=Object Bounds Max Adjustment, sortPriority=15}
   Object Bounds Min Adjustment=FunctionInput(Preview=Constant3Vector() {constant=<rgba 0 5 0 0>}) {bUsePreviewValueAsDefault=true, inputName=Object Bounds Min Adjustment, sortPriority=15}
   Object Bounds Center Adjustment=FunctionInput(Preview=Constant3Vector() {constant=<rgba 0 0 0.145 1>}) {bUsePreviewValueAsDefault=true, inputName=Object Bounds Center Adjustment, sortPriority=15}
   isMech=@n7
   Side Project Vehicle
   Local Position=FunctionInput(Preview=VertexInterpolator(VS=@n8).PS) {bUsePreviewValueAsDefault=true, inputName=Local Position, sortPriority=2}
  ) {materialFunction=&/MF_WrapBounds}
  n10 = Constant3Vector() {constant=<rgba 0 0 -0.25 1>}
  n11 = Constant3Vector() {constant=<rgba 0 0 0 1>}
  n12 = VectorParameter() {group=Wrap_Carnaval, parameterName="CarnB Fill_alternate_Axis_Color_A=Roughness"}
  n13 = TextureObjectParameter() {group=Wrap_Carnaval, parameterName=CarnB_Mask, sortPriority=15, texture=&T_BlacksRGB}
  n14 = FunctionInput(Preview=VertexInterpolator(VS=PreSkinnedNormal()).PS) {bUsePreviewValueAsDefault=true, inputName=Local Normals, sortPriority=1}
  n15 = LinearInterpolate(
   A=LinearInterpolate(
    A=ScalarParameter() {defaultValue=1, group=Wrap_Carnaval, parameterName=CanB_StarsB_Weapon_Scale}
    B=ScalarParameter() {defaultValue=1, group=Wrap_Carnaval, parameterName=CanB_StarsB_Vehicle_Scale}
    Alpha=@n6
   )
   B=ScalarParameter() {defaultValue=1, group=Wrap_Carnaval, parameterName=CanB_StarsB_Mech_Scale}
   Alpha=@n7
  )
  n16 = Time()
  n17 = ScalarParameter() {defaultValue=-9999999, group=Wrap_Carnaval, parameterName=CanB_StarsB_Main_Time}
  n18 = Saturate(None=Divide(A=Subtract(A=@n16, B=@n17), B=ScalarParameter() {defaultValue=0.5, group=Wrap_Carnaval, parameterName=CanB_Stars_Main_Duration}))
  n19 = ScalarParameter() {defaultValue=0.6, group=Wrap_Carnaval, parameterName=CanB_Stars_Main_Curve}
  n20 = Power(Base=Multiply(A=@n18) {constB=0.25}, Exp=@n19)
  n21 = Power(Base=@n18, Exp=@n19)
  n22 = LinearInterpolate(
   A=MaterialFunctionCall(R=@n20, G=@n21, B=@n20) {materialFunction=&AppendMany}.RGB
   B=MaterialFunctionCall(R=@n21, G=@n20, B=@n20) {materialFunction=&AppendMany}.RGB
   Alpha=@n6
  )
  n23 = MaterialFunctionCall(R=@n20, G=@n20, B=@n21) {materialFunction=&AppendMany}
  n24 = MaterialFunctionCall!(
   IsCharacter
   IsStatsB
   Texture=@n13
   Rotation_Axis=LinearInterpolate(A=LinearInterpolate(A=@n10, B=@n11, Alpha=@n6), B=Constant3Vector() {constant=<rgba 0.25 0.25 0 1>}, Alpha=@n7)
   Caculated_Bounds=Add(A=Multiply(A=@n9.Result, B=@n15), B=LinearInterpolate(A=@n22, B=@n23.RGB, Alpha=@n7))
   Pre_Skinned_Normals=@n14
   Blend_Contrast=ScalarParameter() {group=Wrap_Carnaval, parameterName=CarnB_Blend_Contrast}
   Fill_alternate_Axis_Opacity=ScalarParameter() {defaultValue=1, group=Wrap_Carnaval, parameterName=CarnB_Fill_alternate_Axis_Opacity}
   "Fill_alternate_Axis_Color_A=Roughness"=AppendVector(A=@n12.RGB, B=@n12.RGB)
   Is_Vehicle=@n6
   Is_Mech=@n7
   Enable_Fill_Z_Weapon
   Fill_Z_Weapon_opacity
   Fill_Z_Weapon_Color
  ) {materialFunction=&/MF_TopsSideProjection_Base}
  n25 = MaterialFunctionCall(
   Transition Phase=MaterialFunctionCall!(
    ForwardAxis=@n5
    Bounds=@n9.Result
    Pre-Skinned Local Position
    isVehicle=@n7
    isMech=@n6
    Side Project Vehicle
    WeaponOffset=@n3
    WeaponFalloff=@n4
    VehicleOffset=@n3
    VehicleFalloff=@n4
    MechOffset=@n3
    MechFalloff=@n4
    Invert_Gradent
    UV_Spread_XY
    UV_Animated_XY
    Warp_UVs
    Rotate_Mech
    Rotate_Vehicle_Side
    Rotate_Vehicle
   ) {materialFunction=&/MF_Wrap_Gradent_Creation}.Gradent_B
   Height Texture=ChannelMaskParameter(None=@n24.RGB) {desc=uses CarnB_Mask for source, group=Wrap_Carnaval, parameterName=CarnB_StarsB_Mask_Height}.R
   Contrast=ScalarParameter() {defaultValue=5, group=Wrap_Carnaval, parameterName=CarnB_StarsB_Contrast_Height}
  ) {materialFunction=&HeightLerp}
  n26 = OneMinus(None=@n25.Alpha)
  n27 = ScalarParameter() {defaultValue=0.65, group=Wrap_Carnaval, parameterName=CanB_Stars_Opacity_Duration}
  n28 = Divide(A=Subtract(A=Time(), B=ScalarParameter() {defaultValue=-9999999, group=Wrap_Carnaval, parameterName=CanB_StarsB_Opacity_Time}), B=@n27)
  n29 = ScalarParameter() {defaultValue=0.5, group=Wrap_Carnaval, parameterName=CanB_Stars_Opacity_Curve}
  n30 = OneMinus(None=Power(Base=Saturate(None=@n28), Exp=@n29))
  n31 = Time()
  n32 = VectorParameter() {group=Wrap_Carnaval, parameterName=CanB_BackGround_Weapon_Speed, sortPriority=20, defaultValue=<rgba 0 0 0 1>}
  n33 = VectorParameter() {group=Wrap_Carnaval, parameterName=CanB_BackGround_vehicle_Speed, sortPriority=20, defaultValue=<rgba 0 0 0 1>}
  n34 = VectorParameter() {group=Wrap_Carnaval, parameterName=CanB_BackGround_vehicle_SP_Speed, sortPriority=20, defaultValue=<rgba 0 0 0 1>}
  n35 = FunctionInput() {bUsePreviewValueAsDefault=true, inputName=Side Project Vehicle, inputType=FunctionInput_Scalar}
  n36 = VectorParameter() {group=Wrap_Carnaval, parameterName=CanB_BackGround_Mech_Speed, sortPriority=20, defaultValue=<rgba 0 0 0 1>}
  n37 = LinearInterpolate(A=LinearInterpolate(A=@n32.RGB, B=LinearInterpolate(A=@n33.RGB, B=@n34.RGB, Alpha=@n35), Alpha=@n6), B=@n36.RGB, Alpha=@n7)
  n38 = Multiply(A=@n9.Result, B=Constant3Vector() {constant=<rgba 1 1 -1 0>})
  n39 = MaterialFunctionCall(Color=Add(A=@n38, B=Multiply(A=@n31, B=@n37))) {materialFunction=&SplitComponents}
  n40 = AppendVector(A=@n39.G, B=@n39.G)
  n41 = LinearInterpolate(A=AppendVector(A=@n39.R, B=@n39.R), B=AppendVector(A=@n39.R, B=@n39.R), Alpha=@n35)
  n42 = OneMinus(None=@n39.B)
  n43 = TextureObjectParameter() {
   group=Wrap_Carnaval
   parameterName="CarnB_BackGround_BC  A=Roughness"
   sortPriority=20
   texture=&T_BlacksRGB
  }
  n44 = TextureSample(UVs=LinearInterpolate(A=LinearInterpolate(A=@n40, B=@n41, Alpha=@n6), B=AppendVector(A=@n42, B=@n39.R), Alpha=@n7), Tex=@n43)
  n45 = VectorParameter() {group=Wrap_Carnaval, parameterName=CanB_Strip_Weapon_Speed, sortPriority=10, defaultValue=<rgba 0 0 -0.1 1>}
  n46 = VectorParameter() {group=Wrap_Carnaval, parameterName=CanB_Strip_vehicle_Speed, sortPriority=10, defaultValue=<rgba 0 0 0 1>}
  n47 = VectorParameter() {group=Wrap_Carnaval, parameterName=CanB_Strip_vehicle_SP_Speed, sortPriority=10, defaultValue=<rgba 0 0 0 1>}
  n48 = LinearInterpolate(A=@n46.RGB, B=@n47.RGB, Alpha=@n35)
  n49 = VectorParameter() {group=Wrap_Carnaval, parameterName=CanB_Strip_Mech_Speed, sortPriority=10, defaultValue=<rgba 0 0 0 1>}
  n50 = Multiply(A=Time(), B=LinearInterpolate(A=LinearInterpolate(A=@n45.RGB, B=@n48, Alpha=@n6), B=@n49.RGB, Alpha=@n7))
  n51 = ScalarParameter() {defaultValue=0.15, group=Wrap_Carnaval, parameterName=CarnB_Warp_amount, sortPriority=15}
  n52 = Time()
  n53 = VectorParameter() {group=Wrap_Carnaval, parameterName=CanB_Warp_Weapon_Speed, sortPriority=15, defaultValue=<rgba 0 0 0.05 1>}
  n54 = VectorParameter() {group=Wrap_Carnaval, parameterName=CanB_Warp_vehicle_Speed, sortPriority=15, defaultValue=<rgba 0 0 0 1>}
  n55 = VectorParameter() {group=Wrap_Carnaval, parameterName=CanB_Warp_vehicle_SP_Speed, sortPriority=15, defaultValue=<rgba 0 0 0 1>}
  n56 = VectorParameter() {group=Wrap_Carnaval, parameterName=CanB_Warp_Mech_Speed, sortPriority=15, defaultValue=<rgba 0 0 0 1>}
  n57 = LinearInterpolate(A=LinearInterpolate(A=@n53.RGB, B=LinearInterpolate(A=@n54.RGB, B=@n55.RGB, Alpha=@n35), Alpha=@n6), B=@n56.RGB, Alpha=@n7)
  n58 = MaterialFunctionCall(Color=Add(A=@n38, B=Multiply(A=@n52, B=@n57))) {materialFunction=&SplitComponents}
  n59 = LinearInterpolate(
   A=AppendVector(A=@n58.G, B=@n58.G)
   B=LinearInterpolate(A=AppendVector(A=@n58.R, B=@n58.R), B=AppendVector(A=@n58.R, B=@n58.R), Alpha=@n35)
   Alpha=@n6
  )
  n60 = AppendVector(A=OneMinus(None=@n58.B), B=@n58.R)
  n61 = ChannelMaskParameter(None=TextureSample(UVs=LinearInterpolate(A=@n59, B=@n60, Alpha=@n7), Tex=@n13).RGBA) {
   desc=uses CarnB_Mask for source
   group=Wrap_Carnaval
   maskChannel=Alpha
   parameterName=CarnB_Warp_Mask
   sortPriority=15
   defaultValue=<rgba 0 0 0 1>
  }
  n62 = MaterialFunctionCall(Color=Add(A=Add(A=@n50, B=@n38), B=Multiply(A=@n61.R, B=@n51))) {materialFunction=&SplitComponents}
  n63 = AppendVector(A=@n62.G, B=@n62.G)
  n64 = LinearInterpolate(A=AppendVector(A=@n62.R, B=@n62.R), B=AppendVector(A=@n62.R, B=@n62.R), Alpha=@n35)
  n65 = OneMinus(None=@n62.B)
  n66 = TextureObjectParameter() {group=Wrap_Carnaval, parameterName=CarnB_Strip_BC_Alpha, sortPriority=10, texture=&T_BlacksRGB}
  n67 = TextureSample(UVs=LinearInterpolate(A=LinearInterpolate(A=@n63, B=@n64, Alpha=@n6), B=AppendVector(A=@n65, B=@n62.R), Alpha=@n7), Tex=@n66)
  n68 = ScalarParameter() {defaultValue=-3, group=Wrap_Carnaval, parameterName=CarnB_StarsA_Offset, sortPriority=30}
  n69 = ScalarParameter() {defaultValue=1.75, group=Wrap_Carnaval, parameterName=CarnB_StarsA_Falloff, sortPriority=30}
  n70 = Constant3Vector() {constant=<rgba 0 0 -0.25 1>}
  n71 = Constant3Vector() {constant=<rgba 0 0 0 1>}
  n72 = VectorParameter() {group=Wrap_Carnaval, parameterName="CarnB Fill_alternate_Axis_Color_A=Roughness", sortPriority=30}
  n73 = Time()
  n74 = ScalarParameter() {defaultValue=-9999999, group=Wrap_Carnaval, parameterName=CanB_StarsA_Main_Time, sortPriority=30}
  n75 = Saturate(None=Divide(
    A=Subtract(A=@n73, B=@n74)
    B=ScalarParameter() {defaultValue=0.5, group=Wrap_Carnaval, parameterName=CanB_Stars_Main_Duration, sortPriority=30}
   ))
  n76 = ScalarParameter() {defaultValue=0.6, group=Wrap_Carnaval, parameterName=CanB_Stars_Main_Curve, sortPriority=30}
  n77 = Power(Base=Multiply(A=@n75) {constB=0.25}, Exp=@n76)
  n78 = Power(Base=@n75, Exp=@n76)
  n79 = LinearInterpolate(
   A=MaterialFunctionCall(R=@n77, G=@n78, B=@n77) {materialFunction=&AppendMany}.RGB
   B=MaterialFunctionCall(R=@n78, G=@n77, B=@n77) {materialFunction=&AppendMany}.RGB
   Alpha=@n6
  )
  n80 = MaterialFunctionCall(R=@n77, G=@n77, B=@n78) {materialFunction=&AppendMany}
  n81 = LinearInterpolate(
   A=LinearInterpolate(
    A=ScalarParameter() {defaultValue=1, group=Wrap_Carnaval, parameterName=CanB_StarsA_Weapon_Scale, sortPriority=30}
    B=ScalarParameter() {defaultValue=1, group=Wrap_Carnaval, parameterName=CanB_StarsA_Vehicle_Scale, sortPriority=30}
    Alpha=@n6
   )
   B=ScalarParameter() {defaultValue=1, group=Wrap_Carnaval, parameterName=CanB_StarsA_Mech_Scale, sortPriority=30}
   Alpha=@n7
  )
  n82 = MaterialFunctionCall(
   Texture=@n13
   Rotation_Axis=LinearInterpolate(A=LinearInterpolate(A=@n70, B=@n71, Alpha=@n6), B=Constant3Vector() {constant=<rgba 0.25 0.25 0 1>}, Alpha=@n7)
   Caculated_Bounds=Add(A=Multiply(A=@n9.Result, B=@n81), B=LinearInterpolate(A=@n79, B=@n80.RGB, Alpha=@n7))
   Pre_Skinned_Normals=@n14
   Blend_Contrast=ScalarParameter() {group=Wrap_Carnaval, parameterName=CarnB_Blend_Contrast, sortPriority=30}
   Fill_alternate_Axis_Opacity=ScalarParameter() {defaultValue=1, group=Wrap_Carnaval, parameterName=CarnB_Fill_alternate_Axis_Opacity, sortPriority=30}
   "Fill_alternate_Axis_Color_A=Roughness"=AppendVector(A=@n72.RGB, B=@n72.RGB)
   Is_Vehicle=@n6
   Is_Mech=@n7
  ) {materialFunction=&MF_TopsSideProjection_Base}
  n83 = MaterialFunctionCall(
   Transition Phase=MaterialFunctionCall(
    ForwardAxis=@n5
    Bounds=@n9.Result
    isVehicle=@n6
    isMech=@n7
    WeaponOffset=@n68
    WeaponFalloff=@n69
    VehicleOffset=@n68
    VehicleFalloff=@n69
    MechOffset=@n68
    MechFalloff=@n69
   ) {materialFunction=&MF_Wrap_Gradent_Creation}.Gradent_B
   Height Texture=ChannelMaskParameter(None=@n82.RGB) {desc=uses CarnB_Mask for source, group=Wrap_Carnaval, parameterName=CarnB_StarsA_Mask_Height, sortPriority=30}.R
   Contrast=ScalarParameter() {defaultValue=5, group=Wrap_Carnaval, parameterName=CarnB_StarsA_Contrast_Height, sortPriority=30}
  ) {materialFunction=&HeightLerp}
  n84 = Divide(A=Subtract(A=Time(), B=ScalarParameter() {defaultValue=-9999999, group=Wrap_Carnaval, parameterName=CanB_StarsA_Opacity_Time}), B=@n27)
  n85 = OneMinus(None=Power(Base=Saturate(None=@n84), Exp=@n29))
  n86 = Multiply(
   A=Multiply(A=LinearInterpolate(A=@n44.RGB, B=@n67.RGB, Alpha=@n67.A), B=LinearInterpolate(B=@n26, Alpha=@n30) {constA=1})
   B=LinearInterpolate(B=OneMinus(None=@n83.Alpha), Alpha=@n85) {constA=1}
  )
  n87 = LinearInterpolate(A=Constant3Vector() {constant=<rgba 0 0 -0.25 1>}, B=Constant3Vector() {constant=<rgba 0 0 0 1>}, Alpha=@n6)
  n88 = Constant3Vector() {constant=<rgba 0.25 0.25 0 1>}
  n89 = VectorParameter() {group=Wrap_Carnaval, parameterName="CarnB Fill_alternate_Axis_Color_A=Roughness"}
  n90 = ChannelMaskParameter(None=MaterialFunctionCall(
    Texture=TextureObjectParameter() {group=Wrap_Carnaval, parameterName=CarnB_Mask, texture=&T_BlacksRGB}
    Rotation_Axis=LinearInterpolate(A=@n87, B=@n88, Alpha=@n7)
    Caculated_Bounds=@n9.Result
    Pre_Skinned_Normals=@n14
    Blend_Contrast=ScalarParameter() {group=Wrap_Carnaval, parameterName=CarnB_Blend_Contrast}
    Fill_alternate_Axis_Opacity=ScalarParameter() {defaultValue=1, group=Wrap_Carnaval, parameterName=CarnB_Fill_alternate_Axis_Opacity}
    "Fill_alternate_Axis_Color_A=Roughness"=AppendVector(A=@n89.RGB, B=@n89.RGB)
    Is_Vehicle=@n6
    Is_Mech=@n7
   ) {materialFunction=&MF_TopsSideProjection_Base}.RGB) {desc=uses CarnB_Mask for source, group=Wrap_Carnaval, maskChannel=Blue, parameterName=CarnB_HornA_Color_Mask, defaultValue=<rgba 0 0 1 0>}
  n91 = Subtract(A=Time(), B=ScalarParameter() {defaultValue=-9999999, group=Wrap_Carnaval, parameterName=CarnB_HornA_Time})
  n92 = ScalarParameter() {defaultValue=0.25, group=Wrap_Carnaval, parameterName=CarnB_HornA_Duration}
  n93 = LinearInterpolate(
   A=ScalarParameter() {group=Wrap_Carnaval, parameterName=CarnB_HornA_Offset Start}
   B=ScalarParameter() {defaultValue=-1, group=Wrap_Carnaval, parameterName=CarnB_HornA_Offset End}
   Alpha=Saturate(None=Divide(A=@n91, B=@n92))
  )
  n94 = ScalarParameter() {defaultValue=-0.1, group=Wrap_Carnaval, parameterName=CarnB_HornA_Falloff}
  n95 = MaterialFunctionCall(
   ForwardAxis=@n5
   Bounds=@n9.Result
   isVehicle=@n6
   isMech=@n7
   WeaponOffset=@n93
   WeaponFalloff=@n94
   VehicleOffset=@n93
   VehicleFalloff=@n94
   MechOffset=@n93
   MechFalloff=@n94
  ) {materialFunction=&MF_Wrap_Gradent_Creation}
  n96 = Multiply(A=@n90.R, B=@n95.Gradent)
  n97 = Saturate(None=Multiply(A=@n96, B=Add(A=@n96, B=@n95.Gradent)))
  n98 = OneMinus(None=@n97)
  n99 = ScalarParameter() {defaultValue=-0.5, group=Wrap_Carnaval, parameterName=CarnB_HornB_ColorGrad_Offset_Weapon}
  n100 = ScalarParameter() {group=Wrap_Carnaval, parameterName=CarnB_HornB_ColorGrad_Offset_Vehicle}
  n101 = ScalarParameter() {defaultValue=0.5, desc="negative value inverts ", group=Wrap_Carnaval, parameterName=CarnB_HornB_ColorGrad_falloff_Weapon}
  n102 = ScalarParameter() {desc="negative value inverts ", group=Wrap_Carnaval, parameterName=CarnB_HornB_ColorGrad_falloff_Vehicle}
  n103 = MaterialFunctionCall!(
   Bounds=@n9.Result
   ForwardAxis=@n5
   Offset=LinearInterpolate(
    A=LinearInterpolate(A=@n99, B=@n100, Alpha=@n6)
    B=ScalarParameter() {group=Wrap_Carnaval, parameterName=CarnB_HornB_ColorGrad_Offset_Mech}
    Alpha=@n7
   )
   FallOff=LinearInterpolate(
    A=LinearInterpolate(A=@n101, B=@n102, Alpha=@n6)
    B=ScalarParameter() {desc="negative value inverts ", group=Wrap_Carnaval, parameterName=CarnB_HornB_ColorGrad_falloff_Mech}
    Alpha=@n7
   )
  ) {materialFunction=&/MF_Wrap_Gradent_Simple}
  n104 = Multiply(
   A=@n97
   B=LinearInterpolate(
    A=VectorParameter() {group=Wrap_Carnaval, parameterName=CarnB_HornB_ColorA, defaultValue=<rgba 0.9911 0.6654 0.0152 1>}.RGB
    B=VectorParameter() {group=Wrap_Carnaval, parameterName=CarnB_HornB_ColorB, defaultValue=<rgba 0.9911 0.0762 0.1248 1>}.RGB
    Alpha=Saturate(None=@n103.Raw_Data)
   )
  )
  n105 = Saturate(None=Add(A=OneMinus(None=@n67.A), B=ScalarParameter() {defaultValue=1, group=Wrap_Carnaval, parameterName=CarnB_Strip_Roughness}))
  n106 = ChannelMaskParameter(None=@n24.RGB) {desc=uses CarnB_Mask for source, group=Wrap_Carnaval, maskChannel=Green, parameterName=CarnB_StarsB_Color_Mask, defaultValue=<rgba 0 1 0 0>}
  n107 = Multiply(A=@n25.Alpha, B=VectorParameter() {group=Wrap_Carnaval, parameterName=CarnB_StarsB_Color, defaultValue=<rgba 0.1095 0.9911 0.9216 1>}.RGB)
  n108 = Add(
   A=@n2.EmissiveColor
   B=LinearInterpolate(
    B=MaterialFunctionCall(Hue Shift Percentage=@n106.R, Texture=@n107) {materialFunction=&HueShift}.Result
    Alpha=@n30
   ) {constB=0}
  )
  n109 = ChannelMaskParameter(None=@n82.RGB) {
   desc=uses CarnB_Mask for source
   group=Wrap_Carnaval
   maskChannel=Green
   parameterName=CarnB_StarsA_Color_Mask
   sortPriority=30
   defaultValue=<rgba 0 1 0 0>
  }
  n110 = VectorParameter() {group=Wrap_Carnaval, parameterName=CarnB_StarsA_Color, sortPriority=30, defaultValue=<rgba 1 1 0.0666 1>}
  n111 = LinearInterpolate(
   B=MaterialFunctionCall(Hue Shift Percentage=@n109.R, Texture=Multiply(A=@n83.Alpha, B=@n110.RGB)) {materialFunction=&HueShift}.Result
   Alpha=@n85
  ) {constB=0}
  n112 = SetMaterialAttributes!(
   MaterialAttributes=@n2.MaterialAttributes
   BaseColor=Add(A=Multiply(A=@n86, B=@n98), B=Saturate(None=@n104))
   Metallic=Constant()
   Roughness=LinearInterpolate(
    A=ScalarParameter() {group=Wrap_Carnaval, parameterName=CarnB_Roughness_Low}
    B=ScalarParameter() {defaultValue=1, group=Wrap_Carnaval, parameterName=CarnB_Roughness_High}
    Alpha=Multiply(A=@n105, B=@n44.A)
   )
   EmissiveColor=Multiply(A=Add(A=@n108, B=@n111), B=@n98)
  ) {attributeSetTypes [
    #1
    #3
    #4
    #2
   ]}
  n113 = FunctionInput() {inputName="Customization Mask (_CM B)", inputType=FunctionInput_Scalar, sortPriority=24, previewValue=<1 0 0 1>}
  FunctionOutput(None=StaticSwitchParameter(True=BlendMaterialAttributes(A=@n1, B=@n112, Alpha=@n113) {vertexAttributeBlendType=UseB}, False=@n1) {group=Customization_Wrap, parameterName=Carnaval})
  OneMinus(None=@n20)
 ]}
