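# Material function for weapon-skin wraps that layers four animated clockwork gears onto input material attributes, gated by the ClockWork switch and customization mask.
# Side-projected pre-skinned UVs receive offsets, scale and rotation selected by item state (weapon, vehicle, mech, character, backpack, pickaxe).
# Each gear spins within its own quadrant of shared atlas textures.
# MF_ClockWork_Gears composites them, driving color, metallic, roughness, normals over a background, plus optional emissive.
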MaterialFunction /Game/Weapons/WeaponSkins/Wraps/MaterialFunctions/MF_ClockWork
graph {nodes [
  n1 = VectorParameter() {group=Wrap_ClockWork, parameterName=GearB_Position Weapon, defaultValue=<rgba -0.05 -0.1 0 1>}
  n2 = VectorParameter() {group=Wrap_ClockWork, parameterName=GearB_Position, defaultValue=<rgba 0.05 0.1 0 1>}
  n3 = Constant4Vector()
  n4 = MaterialFunctionCall(Float4=NamedRerouteDeclaration(None=FunctionInput(Preview=@n3.RGBA) {bUsePreviewValueAsDefault=true, inputName=IsState, inputType=FunctionInput_Vector4, sortPriority=4}) {name=IsState, nodeColor=<rgba 0.1497 1 0.1633 1>}) {materialFunction=&BreakOutFloat4Components}
  n5 = NamedRerouteDeclaration(None=@n4.R) {name=isVehicle, nodeColor=<rgba 1 0 0.4219 1>}
  n6 = Saturate(None=Add(A=@n5, B=ScalarParameter() {defaultValue=1, group=Wrap_ClockWork, parameterName="GearB_Position --> 0 Weapon/Vehicle Split"}))
  ComponentMask(None=LinearInterpolate(
    A=LinearInterpolate(A=@n1.RGB, B=@n2.RGB, Alpha=@n6)
    B=VectorParameter() {group=Wrap_ClockWork, parameterName=GearB_Position Pickaxe, defaultValue=<rgba 0 0 0 1>}.RGB
    Alpha=NamedRerouteUsage()
   )) {g=true, r=true}
  n7 = VectorParameter() {group=Wrap_ClockWork, parameterName=GearC_Position Weapon, defaultValue=<rgba -0.05 -0.1 0 1>}
  n8 = VectorParameter() {group=Wrap_ClockWork, parameterName=GearC_Position, defaultValue=<rgba 0.1 -0.05 0 1>}
  n9 = Saturate(None=Add(A=@n5, B=ScalarParameter() {defaultValue=1, group=Wrap_ClockWork, parameterName="GearC_Position --> 0 Weapon/Vehicle Split"}))
  ComponentMask(None=LinearInterpolate(
    A=LinearInterpolate(A=@n7.RGB, B=@n8.RGB, Alpha=@n9)
    B=VectorParameter() {group=Wrap_ClockWork, parameterName=GearC_Position Pickaxe, defaultValue=<rgba 0 0 0 1>}.RGB
    Alpha=NamedRerouteUsage()
   )) {g=true, r=true}
  n10 = VectorParameter() {group=Wrap_ClockWork, parameterName=GearD_Position Weapon, defaultValue=<rgba -0.05 -0.1 0 1>}
  n11 = VectorParameter() {group=Wrap_ClockWork, parameterName=GearD_Position, defaultValue=<rgba -0.05 -0.1 0 1>}
  n12 = Saturate(None=Add(A=ScalarParameter() {defaultValue=1, group=Wrap_ClockWork, parameterName="GearD_Position --> 0 Weapon/Vehicle Split"}, B=@n5))
  ComponentMask(None=LinearInterpolate(
    A=LinearInterpolate(A=@n10.RGB, B=@n11.RGB, Alpha=@n12)
    B=VectorParameter() {group=Wrap_ClockWork, parameterName=GearD_Position Pickaxe, defaultValue=<rgba 0 0 0 1>}.RGB
    Alpha=NamedRerouteUsage()
   )) {g=true, r=true}
  ComponentMask(None=Constant3Vector()) {g=true, r=true}
  Desaturation()
  n13 = TextureObjectParameter() {
   desc=Red Gear mask Green Roughness
   group=Wrap_ClockWork
   parameterName=Mask_ClockWork
   samplerType=SAMPLERTYPE_Masks
   sortPriority=-1
   texture=/Game/Weapons/WeaponSkins/Wraps/Textures/ClockWork/T_Wrap_ClockWork_M
  }
  n14 = Time()
  n15 = ComponentMask(None=VectorParameter() {group=Wrap_ClockWork, parameterName=GearA-Position--Speed}.RGB) {g=true, r=true}
  n16 = NamedRerouteUsage()
  n17 = Ceil(None=ScalarParameter() {group=Wrap_ClockWork_Uvs, parameterName="CW_Offset 1= befor Scale", sliderMax=1, sortPriority=0})
  n18 = LinearInterpolate(A=@n16, B=Add(A=@n16, B=NamedRerouteUsage()), Alpha=@n17)
  n19 = NamedRerouteUsage()
  n20 = StaticSwitchParameter(
   True=MaterialFunctionCall(UVs=@n18, Texture Scale=@n19) {materialFunction=&ScaleUVsByCenter}.UVs
   False=Multiply(A=@n19, B=@n18)
  ) {group=Wrap_ClockWork, parameterName=TileRate_From_Center}
  n21 = NamedRerouteUsage()
  n22 = VectorParameter() {group=Wrap_ClockWork_Uvs, parameterName=Gear Rotation_Centerpoint, sortPriority=0, defaultValue=<rgba 0.5 0.5 0 1>}
  n23 = MaterialFunctionCall(
   UVs=LinearInterpolate(A=Add(A=@n20, B=@n21), B=@n20, Alpha=@n17)
   Rotation Center=ComponentMask(None=@n22.RGB) {g=true, r=true}
   "Rotation Angle (0-1)"=NamedRerouteUsage()
  ) {materialFunction=&CustomRotator}
  n24 = Add(A=@n23.Rotated Values, B=Add(A=NamedRerouteUsage(), B=Multiply(A=@n14, B=@n15)))
  n25 = ScalarParameter() {defaultValue=1, group=Wrap_ClockWork, parameterName=GearA_Scale}
  n26 = ScalarParameter() {defaultValue=4, group=Wrap_ClockWork_Uvs, parameterName=Master_Gear_Scale, sortPriority=0}
  n27 = MaterialFunctionCall(UVs=@n24, Texture Scale=Divide(A=@n25, B=@n26)) {materialFunction=&ScaleUVsByCenter}
  n28 = Frac(None=@n27.UVs)
  n29 = Constant2Vector()
  n30 = ScalarParameter() {defaultValue=0.09, group=Wrap_ClockWork, parameterName=GearA_Rotation_Speed}
  n31 = Multiply(A=Time(), B=LinearInterpolate(A=Multiply(A=@n30) {constB=-1}, B=@n30, Alpha=NamedRerouteUsage()))
  n32 = MaterialFunctionCall(
   UVs=Add(A=Multiply(A=@n28) {constB=0.5}, B=@n29)
   Rotation Center=Add(A=@n29, B=Constant2Vector() {g=0.25, r=0.25})
   "Rotation Angle (0-1)"=@n31
  ) {materialFunction=&CustomRotator}
  n33 = Time()
  n34 = ComponentMask(None=VectorParameter() {group=Wrap_ClockWork, parameterName=GearB-Position--Speed}.RGB) {g=true, r=true}
  n35 = Add(A=@n23.Rotated Values, B=Add(A=NamedRerouteUsage(), B=Multiply(A=@n33, B=@n34)))
  n36 = ScalarParameter() {defaultValue=1, group=Wrap_ClockWork, parameterName=GearB_Scale}
  n37 = MaterialFunctionCall(UVs=@n35, Texture Scale=Divide(A=@n36, B=@n26)) {materialFunction=&ScaleUVsByCenter}
  n38 = Frac(None=@n37.UVs)
  n39 = Constant2Vector() {g=0.5}
  n40 = ScalarParameter() {defaultValue=-0.1, group=Wrap_ClockWork, parameterName=GearB_Rotation_Speed}
  n41 = Multiply(A=Time(), B=LinearInterpolate(A=Multiply(A=@n40) {constB=-1}, B=@n40, Alpha=NamedRerouteUsage()))
  n42 = MaterialFunctionCall(
   UVs=Add(A=Multiply(A=@n38) {constB=0.5}, B=@n39)
   Rotation Center=Add(A=@n39, B=Constant2Vector() {g=0.25, r=0.25})
   "Rotation Angle (0-1)"=@n41
  ) {materialFunction=&CustomRotator}
  n43 = Time()
  n44 = ComponentMask(None=VectorParameter() {group=Wrap_ClockWork, parameterName=GearC-Position--Speed}.RGB) {g=true, r=true}
  n45 = Add(A=@n23.Rotated Values, B=Add(A=NamedRerouteUsage(), B=Multiply(A=@n43, B=@n44)))
  n46 = ScalarParameter() {defaultValue=1, group=Wrap_ClockWork, parameterName=GearC_Scale}
  n47 = MaterialFunctionCall(UVs=@n45, Texture Scale=Divide(A=@n46, B=@n26)) {materialFunction=&ScaleUVsByCenter}
  n48 = Time()
  n49 = ComponentMask(None=VectorParameter() {group=Wrap_ClockWork, parameterName=GearD-Position--Speed}.RGB) {g=true, r=true}
  n50 = Add(A=@n23.Rotated Values, B=Add(A=NamedRerouteUsage(), B=Multiply(A=@n48, B=@n49)))
  n51 = ScalarParameter() {defaultValue=1, group=Wrap_ClockWork, parameterName=GearD_Scale}
  n52 = MaterialFunctionCall(UVs=@n50, Texture Scale=Divide(A=@n51, B=@n26)) {materialFunction=&ScaleUVsByCenter}
  n53 = Frac(None=@n52.UVs)
  n54 = Constant2Vector() {g=0.5, r=0.5}
  n55 = ScalarParameter() {defaultValue=-0.48, group=Wrap_ClockWork, parameterName=GearD_Rotation_Speed}
  n56 = Multiply(A=Time(), B=LinearInterpolate(A=Multiply(A=@n55) {constB=-1}, B=@n55, Alpha=NamedRerouteUsage()))
  n57 = MaterialFunctionCall(
   UVs=Add(A=Multiply(A=@n53) {constB=0.5}, B=@n54)
   Rotation Center=Add(A=@n54, B=Constant2Vector() {g=0.25, r=0.25})
   "Rotation Angle (0-1)"=@n56
  ) {materialFunction=&CustomRotator}
  n58 = Frac(None=@n47.UVs)
  n59 = Constant2Vector() {r=0.5}
  n60 = ScalarParameter() {defaultValue=0.43, group=Wrap_ClockWork, parameterName=GearC_Rotation_Speed}
  n61 = Multiply(A=Time(), B=LinearInterpolate(A=Multiply(A=@n60) {constB=-1}, B=@n60, Alpha=NamedRerouteUsage()))
  n62 = MaterialFunctionCall(
   UVs=Add(A=Multiply(A=@n58) {constB=0.5}, B=@n59)
   Rotation Center=Add(A=@n59, B=Constant2Vector() {g=0.25, r=0.25})
   "Rotation Angle (0-1)"=@n61
  ) {materialFunction=&CustomRotator}
  n63 = MaterialFunctionCall!(
   BackGroundMask
   Gear_mask_Red=@n13
   Gears=@n13
   Normals=@n13
   Uvs_GearA_Mask=@n27.UVs
   Uvs_GearA=@n32.Rotated Values
   Uvs_GearB_Mask=@n37.UVs
   Uvs_GearB=@n42.Rotated Values
   Uvs_GearC_Mask=@n47.UVs
   Uvs_GearC=@n62.Rotated Values
   Uvs_GearD_Mask=@n52.UVs
   Uvs_GearD=@n57.Rotated Values
   Local Normals=NamedRerouteUsage()
   IsState=NamedRerouteUsage()
   Rotation=NamedRerouteUsage()
  ) {materialFunction=&/MF_ClockWork_Gears}
  Desaturation(None=@n63.Emissive)
  n64 = Constant()
  FunctionInput(Preview=@n64) {bUsePreviewValueAsDefault=true, inputName=isMech, inputType=FunctionInput_Scalar}
  FunctionInput(Preview=@n64) {bUsePreviewValueAsDefault=true, inputName=Side Project Vehicle, inputType=FunctionInput_Scalar}
  n65 = FunctionInput() {inputName=Material, inputType=FunctionInput_MaterialAttributes}
  n66 = GetMaterialAttributes(None=@n65) {attributeGetTypes [#1]}
  n67 = ScalarParameter() {defaultValue=1, group=Wrap_ClockWork, parameterName=ClockWork_Darken_Base, sortPriority=-3}
  n68 = TextureObjectParameter() {
   group=Wrap_ClockWork
   parameterName=BC_Clockwork
   sortPriority=-1
   texture=/Game/Weapons/WeaponSkins/Wraps/Textures/ClockWork/T_Wrap_ClockWork_BC
  }
  n69 = TextureObjectParameter() {
   group=Wrap_ClockWork
   parameterName=N_ClockWork
   samplerType=SAMPLERTYPE_Normal
   sortPriority=-1
   texture=/Game/Weapons/WeaponSkins/Wraps/Textures/ClockWork/T_Wrap_ClockWork_N
  }
  n70 = MaterialFunctionCall(
   BackGroundMask=Constant3Vector() {constant=<rgba 0 0 1 0>}
   Gear_mask_Red=@n13
   Gears=@n68
   Normals=@n69
   Uvs_GearA_Mask=@n27.UVs
   Uvs_GearA=@n32.Rotated Values
   Uvs_GearB_Mask=@n37.UVs
   Uvs_GearB=@n42.Rotated Values
   Uvs_GearC_Mask=@n47.UVs
   Uvs_GearC=@n62.Rotated Values
   Uvs_GearD_Mask=@n52.UVs
   Uvs_GearD=@n57.Rotated Values
   Local Normals=NamedRerouteUsage()
   IsState=NamedRerouteUsage()
   Rotation=NamedRerouteUsage()
  ) {materialFunction=&MF_ClockWork_Gears}
  n71 = StaticSwitchParameter(True=@n70.Gear_Mask, False=@n63.Emissive) {group=Wrap_ClockWork, parameterName=ClockWork_Diffuse as Emissive, sortPriority=-10}
  n72 = OneMinus(None=Multiply(A=NamedRerouteUsage(), B=Multiply(A=@n67, B=@n71)))
  n73 = SetMaterialAttributes!(MaterialAttributes=@n66.MaterialAttributes, BaseColor=Multiply(A=@n66.BaseColor, B=Saturate(None=@n72))) {attributeSetTypes [#1]}
  n74 = GetMaterialAttributes(None=@n65) {attributeGetTypes [#1, #2, #3]}
  n75 = VectorParameter() {desc=Red Rougness, group=Wrap_ClockWork, parameterName=BackGroundColor, defaultValue=<rgba 0 0 0 1>}
  n76 = ScalarParameter() {group=Wrap_ClockWork, parameterName=BackGround_Spec Min}
  n77 = ScalarParameter() {defaultValue=0.5, group=Wrap_ClockWork, parameterName=BackGround_Spec Max}
  n78 = MaterialFunctionCall(
   Gear_mask_Red=@n13
   Gears=@n68
   Normals=@n13
   Uvs_GearA_Mask=@n27.UVs
   Uvs_GearA=@n32.Rotated Values
   Uvs_GearB_Mask=@n37.UVs
   Uvs_GearB=@n42.Rotated Values
   Uvs_GearC_Mask=@n47.UVs
   Uvs_GearC=@n62.Rotated Values
   Uvs_GearD_Mask=@n52.UVs
   Uvs_GearD=@n57.Rotated Values
   Local Normals=NamedRerouteUsage()
   IsState=NamedRerouteUsage()
   Rotation=NamedRerouteUsage()
  ) {materialFunction=&MF_ClockWork_Gears}
  n79 = Multiply(A=@n78.Gear_Mask_Depth, B=@n78.Gear_Mask_Depth)
  n80 = NamedRerouteUsage()
  n81 = SetMaterialAttributes!(
   MaterialAttributes=@n74.MaterialAttributes
   BaseColor=@n75.RGB
   Metallic=ScalarParameter() {group=Wrap_ClockWork, parameterName=BackGroundMetalic}
   Roughness=@n75.RGB
   Specular=LinearInterpolate(A=@n76, B=LinearInterpolate(A=@n76, B=@n77, Alpha=@n79) {constB=0.5}, Alpha=@n80)
  ) {attributeSetTypes [
    #1
    #3
    #4
    #5
   ]}
  n82 = ScalarParameter() {defaultValue=1, group=Wrap_ClockWork, parameterName=CW-Metalic-------, sortPriority=-3}
  n83 = ChannelMaskParameter(None=@n78.Texture_RGBA) {group=Wrap_ClockWork, maskChannel=Green, parameterName=CW_Metalic_Channel, sortPriority=-3, defaultValue=<rgba 0 1 0 0>}
  n84 = ScalarParameter() {group=Wrap_ClockWork, parameterName="CW-Metalic-UseMask--1=on", sortPriority=-3}
  n85 = ChannelMaskParameter(None=@n78.Texture_RGBA) {group=Wrap_ClockWork, maskChannel=Green, parameterName=CW_Roughness_Channel, sortPriority=-3, defaultValue=<rgba 0 1 0 0>}
  n86 = LinearInterpolate(
   A=@n74.Normal
   B=GetMaterialAttributes(None=FunctionInput() {inputName=Material_Root, inputType=FunctionInput_MaterialAttributes}) {attributeGetTypes [#2]}.Normal
   Alpha=Ceil(None=ScalarParameter() {group=Wrap_ClockWork, parameterName="N_ClockWork_Normals_Msked 1=on", sliderMax=1, sortPriority=-1})
  )
  n87 = SetMaterialAttributes!(
   MaterialAttributes=@n74.MaterialAttributes
   BaseColor=@n70.Combined_Output
   Metallic=Multiply(A=@n70.Gear_Mask, B=LinearInterpolate(A=@n82, B=@n83.R, Alpha=@n84))
   Roughness=@n85.R
   Normal=StaticSwitchParameter(
    True=@n74.Normal
    False=MaterialFunctionCall(BaseNormal=@n86, AdditionalNormal=@n70.Texture_Normals) {materialFunction=&BlendAngleCorrectedNormals}.Result
   ) {group=Wrap_ClockWork, parameterName=N_ClockWork_Disable_Normals, sortPriority=-1}
   Specular=ScalarParameter() {defaultValue=0.5, group=Wrap_ClockWork, parameterName=CW_Specular, sortPriority=-3}
  ) {attributeSetTypes [
    #1
    #3
    #4
    #2
    #5
   ]}
  n88 = BlendMaterialAttributes(
   A=StaticSwitchParameter(True=@n74.MaterialAttributes, False=@n81) {group=Wrap_ClockWork, parameterName=BackGround-Color-Disable--}
   B=@n87
   Alpha=Multiply(A=NamedRerouteUsage(), B=@n79)
  ) {vertexAttributeBlendType=UseB}
  n89 = GetMaterialAttributes(None=StaticSwitchParameter(
    True=@n73
    False=StaticSwitchParameter(True=@n73, False=@n88) {group=Wrap_ClockWork, parameterName=ClockWork_BaseColor_Disable, sortPriority=-5}
   ) {group=Wrap_ClockWork, parameterName=ClockWork_Diffuse as Emissive, sortPriority=-10}) {attributeGetTypes [#6]}
  n90 = ScalarParameter() {defaultValue=1, group=Wrap_ClockWork, parameterName=ClockWork_DiffuseEmissive Strength, sortPriority=-3}
  n91 = Multiply(
   A=NamedRerouteUsage()
   B=StaticSwitchParameter(True=Multiply(A=@n90, B=@n70.Combined_Output), False=@n63.Emissive) {group=Wrap_ClockWork, parameterName=ClockWork_Diffuse as Emissive, sortPriority=-10}
  )
  n92 = StaticSwitchParameter(
   True=SetMaterialAttributes!(MaterialAttributes=@n89.MaterialAttributes, EmissiveColor=Add(A=@n89.EmissiveColor, B=@n91)) {attributeSetTypes [#6]}
   False=@n88
  ) {group=Wrap_ClockWork, parameterName=ClockWork_Emissive, sortPriority=-10}
  n93 = FunctionInput() {inputName="Customization Mask (_CM B)", inputType=FunctionInput_Scalar, sortPriority=24, previewValue=<1 0 0 1>}
  FunctionOutput(None=StaticSwitchParameter(True=BlendMaterialAttributes(A=@n65, B=@n92, Alpha=@n93), False=@n65) {group=Customization, parameterName=ClockWork})
  n94 = VertexInterpolator(VS=PreSkinnedNormal())
  n95 = MaterialFunctionCall(Color=NamedRerouteDeclaration(None=FunctionInput(Preview=@n94.PS) {bUsePreviewValueAsDefault=true, inputName=Local Normals, sortPriority=1}) {name=PreSkin_Normal, nodeColor=<rgba 0.8281 1 0 1>}) {materialFunction=&SplitComponents}
  n96 = Abs(None=@n95.R)
  n97 = ScalarParameter() {defaultValue=0.5, group=Wrap_ClockWork, parameterName=Tri_Mask_Contrast, sortPriority=-4}
  n98 = MaterialFunctionCall(In=Abs(None=@n95.B), Contrast=@n97) {materialFunction=&CheapContrast}
  n99 = MaterialFunctionCall(In=Abs(None=@n95.G), Contrast=@n97) {materialFunction=&CheapContrast}
  LinearInterpolate(
   A=LinearInterpolate(
    A=MaterialFunctionCall(In=@n96, Contrast=@n97) {materialFunction=&CheapContrast}.Result
    B=LinearInterpolate(A=@n98.Result, B=@n99.Result, Alpha=@n4.G)
    Alpha=@n5
   )
   B=@n99.Result
   Alpha=@n4.B
  )
  n100 = Add(A=@n5, B=ScalarParameter() {defaultValue=1, group=Wrap_ClockWork, parameterName="GearA_Position --> 0 Weapon/Vehicle Split"})
  LinearInterpolate(
   A=LinearInterpolate(
    A=VectorParameter() {group=Wrap_ClockWork, parameterName=GearA_Position Weapon, defaultValue=<rgba -0.05 -0.1 0 1>}.RGB
    B=VectorParameter() {group=Wrap_ClockWork, parameterName=GearA_Position}.RGB
    Alpha=Saturate(None=@n100)
   )
   B=VectorParameter() {group=Wrap_ClockWork, parameterName=GearA_Position PickAxe, defaultValue=<rgba 0 0 0 1>}.RGB
   Alpha=NamedRerouteUsage()
  )
  LinearInterpolate(A=@n75.A, B=@n75.A, Alpha=@n80)
  n101 = PreSkinnedPosition()
  n102 = MaterialFunctionCall!(
   is Vehicle=@n5
   Object Bounds Max Adjustment=FunctionInput() {bUsePreviewValueAsDefault=true, inputName=Object Bounds Max Adjustment, sortPriority=15}
   Object Bounds Min Adjustment=FunctionInput() {bUsePreviewValueAsDefault=true, inputName=Object Bounds Min Adjustment, sortPriority=15}
   Object Bounds Center Adjustment=FunctionInput() {bUsePreviewValueAsDefault=true, inputName=Object Bounds Center Adjustment, sortPriority=15}
   isMech=@n4.B
   Side Project Vehicle
   Local Position=FunctionInput(Preview=VertexInterpolator(VS=@n101).PS) {bUsePreviewValueAsDefault=true, inputName=Local Position, sortPriority=2}
  ) {materialFunction=&/MF_WrapBounds}
  n103 = ComponentMask(None=@n102.Result) {g=true}
  n104 = ComponentMask(None=@n102.Result) {b=true}
  n105 = ComponentMask(None=@n102.Result) {r=true}
  n106 = AppendVector(A=OneMinus(None=@n103), B=@n105)
  n107 = AppendVector(A=OneMinus(None=@n104), B=@n105)
  LinearInterpolate(
   A=LinearInterpolate(A=AppendVector(A=@n103, B=@n104), B=LinearInterpolate(A=@n106, B=@n107, Alpha=@n4.G), Alpha=@n5)
   B=AppendVector(A=@n104, B=@n105)
   Alpha=@n4.B
  )
  n108 = TextureSample(UVs=@n57.Rotated Values, Tex=@n69)
  n109 = ComponentMask(None=NamedRerouteUsage()) {a=true}
  n110 = NamedRerouteUsage()
  n111 = NamedRerouteUsage()
  MaterialFunctionCall!(
   Normal_Map
   Normals_Source=@n108.RGB
   Normals_Source_Use=StaticBool() {value=true}
   UVs
   Rotation_Comp=Constant() {r=0.25}
   UV_Rotation=@n109
   Pre_Skinned_Normals=@n110
   is_State=@n111
   "Left\\Right"
   No_Rotation
  ) {materialFunction=&/MF_UVs_Side_Fix_Normals}
  n112 = MaterialFunctionCall(Color=NamedRerouteUsage()) {materialFunction=&SplitComponents}
  n113 = MaterialFunctionCall(Float4=NamedRerouteUsage()) {materialFunction=&BreakOutFloat4Components}
  n114 = LinearInterpolate(A=Multiply(B=@n112.R) {constA=-1}, B=LinearInterpolate(A=@n112.B, B=@n112.B, Alpha=@n113.G), Alpha=@n113.R)
  n115 = Saturate(None=Divide(A=LinearInterpolate(A=@n114, B=@n112.G, Alpha=@n113.B)) {constB=0.001})
  MaterialFunctionCall(In=@n115, Contrast=Constant() {r=5}) {materialFunction=&CheapContrast}
  Multiply(A=@n115)
  Multiply(A=Multiply(A=@n35, B=@n36)) {constB=4}
  Multiply(A=Multiply(A=@n45, B=@n46)) {constB=4}
  Multiply(A=Multiply(A=@n24, B=@n25))
  Multiply(A=Multiply(A=@n50, B=@n51)) {constB=4}
  Multiply(A=MaterialFunctionCall!(Normals input=@n108.RGB, Rotation_Value=@n109, IsState=@n111, Pre_Skinned_Normals=@n110) {materialFunction=&/MF_Projection_Side_NormalFix}.Normal)
  NamedRerouteDeclaration(None=@n4.A) {name=IsCharacter, nodeColor=<rgba 0.3125 0 1 1>}
  n116 = Constant2Vector()
  n117 = MaterialFunctionCall(Float2=NamedRerouteDeclaration(None=FunctionInput(Preview=@n116) {bUsePreviewValueAsDefault=true, inputName=IsState_B, inputType=FunctionInput_Vector2, sortPriority=4}) {name=IsState_B, nodeColor=<rgba 0.1497 1 0.1633 1>}) {materialFunction=&BreakOutFloat2Components}
  NamedRerouteDeclaration(None=@n117.R) {name=IsBackpack, nodeColor=<rgba 0.2656 0 1 1>}
  n118 = VectorParameter() {group=Wrap_ClockWork_Uvs, parameterName=Offset_WeaponCW, defaultValue=<rgba 0 0 0 1>}
  n119 = MaterialFunctionCall(Float4=NamedRerouteUsage()) {materialFunction=&BreakOutFloat4Components}
  n120 = LinearInterpolate(
   A=VectorParameter() {group=Wrap_ClockWork_Uvs, parameterName=Offset_VehicleCW, defaultValue=<rgba 0 0 0 1>}.RGB
   B=VectorParameter() {group=Wrap_ClockWork_Uvs, parameterName=Offset_VehicleCWSP, defaultValue=<rgba 0 0 0 1>}.RGB
   Alpha=@n119.G
  )
  n121 = VectorParameter() {group=Wrap_ClockWork_Uvs, parameterName=Offset_MechCW, defaultValue=<rgba 0 0 0 1>}
  n122 = VectorParameter() {group=Wrap_ClockWork_Uvs, parameterName=Offset_CharacterCW, defaultValue=<rgba 0 0 0 1>}
  n123 = MaterialFunctionCall(Float2=NamedRerouteUsage()) {materialFunction=&BreakOutFloat2Components}
  n124 = LinearInterpolate(
   A=LinearInterpolate(A=LinearInterpolate(A=@n118.RGB, B=@n120, Alpha=@n119.R), B=LinearInterpolate(A=@n121.RGB, B=@n122.RGB, Alpha=@n119.A), Alpha=@n119.B)
   B=VectorParameter() {group=Wrap_ClockWork_Uvs, parameterName=Offset_BackpackCW, defaultValue=<rgba 0 0 0 1>}.RGB
   Alpha=@n123.R
  )
  n125 = VectorParameter() {group=Wrap_ClockWork_Uvs, parameterName=Offset_PickaxeCW, defaultValue=<rgba 0 0 0 1>}
  NamedRerouteDeclaration(None=ComponentMask(None=LinearInterpolate(A=@n124, B=@n125.RGB, Alpha=@n123.G)) {g=true, r=true}) {name=UVs_Offset, nodeColor=<rgba 0 0.625 1 1>}
  n126 = VectorParameter() {group=Wrap_ClockWork, parameterName="GearA_Position A 0=use States", defaultValue=<rgba 0 0 0 1>}
  n127 = MaterialFunctionCall(Float4=NamedRerouteUsage()) {materialFunction=&BreakOutFloat4Components}
  n128 = LinearInterpolate(
   A=VectorParameter() {group=Wrap_ClockWork, parameterName=GearA_Position Weapon, defaultValue=<rgba -0.05 -0.1 0 1>}.RGB
   B=LinearInterpolate(A=@n126.RGB, B=VectorParameter() {group=Wrap_ClockWork, parameterName=GearA_Position_VehSide}.RGB, Alpha=@n127.G)
   Alpha=@n127.R
  )
  n129 = LinearInterpolate(
   A=VectorParameter() {group=Wrap_ClockWork, parameterName=GearA_Position_Mech}.RGB
   B=VectorParameter() {group=Wrap_ClockWork, parameterName=GearA_Position_Character}.RGB
   Alpha=@n127.A
  )
  n130 = MaterialFunctionCall(Float2=NamedRerouteUsage()) {materialFunction=&BreakOutFloat2Components}
  n131 = LinearInterpolate(
   A=LinearInterpolate(
    A=LinearInterpolate(A=@n128, B=@n129, Alpha=@n127.B)
    B=VectorParameter() {group=Wrap_ClockWork, parameterName=GearA_Position_Backpack}.RGB
    Alpha=@n130.R
   )
   B=VectorParameter() {group=Wrap_ClockWork, parameterName=GearA_Position PickAxe, defaultValue=<rgba 0 0 0 1>}.RGB
   Alpha=@n130.G
  )
  NamedRerouteDeclaration(None=ComponentMask(None=LinearInterpolate(A=@n131, B=@n126.RGB, Alpha=@n126.RGB)) {g=true, r=true}) {name=UVs_Offset_A, nodeColor=<rgba 0 0.625 1 1>}
  n132 = ScalarParameter() {defaultValue=1, group=Wrap_ClockWork_Uvs, parameterName=TileRate_VehicleCW}
  n133 = ScalarParameter() {defaultValue=1, group=Wrap_ClockWork_Uvs, parameterName=TileRate_VehicleSideCW}
  n134 = MaterialFunctionCall(Float4=NamedRerouteUsage()) {materialFunction=&BreakOutFloat4Components}
  n135 = LinearInterpolate(
   A=LinearInterpolate(
    A=ScalarParameter() {defaultValue=1, group=Wrap_ClockWork_Uvs, parameterName=TileRate_WeaponCW}
    B=LinearInterpolate(A=@n132, B=@n133, Alpha=@n134.G)
    Alpha=@n134.R
   )
   B=LinearInterpolate(
    A=ScalarParameter() {defaultValue=1, group=Wrap_ClockWork_Uvs, parameterName=TileRate_MechCW}
    B=ScalarParameter() {defaultValue=1, group=Wrap_ClockWork_Uvs, parameterName=TileRate_Character_CW}
    Alpha=@n134.A
   )
   Alpha=@n134.B
  )
  n136 = ScalarParameter() {defaultValue=1, group=Wrap_ClockWork_Uvs, parameterName=TileRate_Backpack_CW}
  n137 = MaterialFunctionCall(Float2=NamedRerouteUsage()) {materialFunction=&BreakOutFloat2Components}
  NamedRerouteDeclaration(None=LinearInterpolate(
    A=LinearInterpolate(A=@n135, B=@n136, Alpha=@n137.R)
    B=ScalarParameter() {defaultValue=1, group=Wrap_ClockWork_Uvs, parameterName=TileRate_Pickaxe_CW}
    Alpha=@n137.G
   )) {name=UVs_Scale_V1, nodeColor=<rgba 0 0.7188 1 1>}
  n138 = VectorParameter() {group=Wrap_ClockWork, parameterName="GearB_Position A 0=use States", defaultValue=<rgba 0.05 0.1 0 1>}
  n139 = MaterialFunctionCall(Float4=NamedRerouteUsage()) {materialFunction=&BreakOutFloat4Components}
  n140 = LinearInterpolate(
   A=VectorParameter() {group=Wrap_ClockWork, parameterName=GearB_Position Weapon, defaultValue=<rgba -0.05 -0.1 0 1>}.RGB
   B=LinearInterpolate(A=@n138.RGB, B=VectorParameter() {group=Wrap_ClockWork, parameterName=GearB_Position_VehSide}.RGB, Alpha=@n139.G)
   Alpha=@n139.R
  )
  n141 = LinearInterpolate(
   A=VectorParameter() {group=Wrap_ClockWork, parameterName=GearB_Position_Mech}.RGB
   B=VectorParameter() {group=Wrap_ClockWork, parameterName=GearB_Position_Character}.RGB
   Alpha=@n139.A
  )
  n142 = MaterialFunctionCall(Float2=NamedRerouteUsage()) {materialFunction=&BreakOutFloat2Components}
  n143 = LinearInterpolate(
   A=LinearInterpolate(
    A=LinearInterpolate(A=@n140, B=@n141, Alpha=@n139.B)
    B=VectorParameter() {group=Wrap_ClockWork, parameterName=GearB_Position_Backpack}.RGB
    Alpha=@n142.R
   )
   B=VectorParameter() {group=Wrap_ClockWork, parameterName=GearB_Position Pickaxe, defaultValue=<rgba 0 0 0 1>}.RGB
   Alpha=@n142.G
  )
  NamedRerouteDeclaration(None=ComponentMask(None=LinearInterpolate(A=@n143, B=@n138.RGB, Alpha=@n138.RGB)) {g=true, r=true}) {name=UVs_Offset_B, nodeColor=<rgba 0 0.625 1 1>}
  n144 = VectorParameter() {group=Wrap_ClockWork, parameterName="GearC_Position A 0=use States", defaultValue=<rgba 0.1 -0.05 0 1>}
  n145 = MaterialFunctionCall(Float4=NamedRerouteUsage()) {materialFunction=&BreakOutFloat4Components}
  n146 = LinearInterpolate(
   A=VectorParameter() {group=Wrap_ClockWork, parameterName=GearC_Position Weapon, defaultValue=<rgba -0.05 -0.1 0 1>}.RGB
   B=LinearInterpolate(A=@n144.RGB, B=VectorParameter() {group=Wrap_ClockWork, parameterName=GearC_Position_VehSide}.RGB, Alpha=@n145.G)
   Alpha=@n145.R
  )
  n147 = LinearInterpolate(
   A=VectorParameter() {group=Wrap_ClockWork, parameterName=GearC_Position_Mech}.RGB
   B=VectorParameter() {group=Wrap_ClockWork, parameterName=GearC_Position_Character}.RGB
   Alpha=@n145.A
  )
  n148 = MaterialFunctionCall(Float2=NamedRerouteUsage()) {materialFunction=&BreakOutFloat2Components}
  n149 = LinearInterpolate(
   A=LinearInterpolate(
    A=LinearInterpolate(A=@n146, B=@n147, Alpha=@n145.B)
    B=VectorParameter() {group=Wrap_ClockWork, parameterName=GearC_Position_Backpack}.RGB
    Alpha=@n148.R
   )
   B=VectorParameter() {group=Wrap_ClockWork, parameterName=GearC_Position Pickaxe, defaultValue=<rgba 0 0 0 1>}.RGB
   Alpha=@n148.G
  )
  NamedRerouteDeclaration(None=ComponentMask(None=LinearInterpolate(A=@n149, B=@n144.RGB, Alpha=@n144.RGB)) {g=true, r=true}) {name=UVs_Offset_C, nodeColor=<rgba 0 0.625 1 1>}
  n150 = VectorParameter() {group=Wrap_ClockWork, parameterName="GearD_Position  A 0=use States", defaultValue=<rgba -0.05 -0.1 0 1>}
  n151 = MaterialFunctionCall(Float4=NamedRerouteUsage()) {materialFunction=&BreakOutFloat4Components}
  n152 = LinearInterpolate(
   A=VectorParameter() {group=Wrap_ClockWork, parameterName=GearD_Position Weapon, defaultValue=<rgba -0.05 -0.1 0 1>}.RGB
   B=LinearInterpolate(A=@n150.RGB, B=VectorParameter() {group=Wrap_ClockWork, parameterName=GearD_Position_VehSide}.RGB, Alpha=@n151.G)
   Alpha=@n151.R
  )
  n153 = LinearInterpolate(
   A=VectorParameter() {group=Wrap_ClockWork, parameterName=GearD_Position_Mech}.RGB
   B=VectorParameter() {group=Wrap_ClockWork, parameterName=GearD_Position_Character}.RGB
   Alpha=@n151.A
  )
  n154 = MaterialFunctionCall(Float2=NamedRerouteUsage()) {materialFunction=&BreakOutFloat2Components}
  n155 = LinearInterpolate(
   A=LinearInterpolate(
    A=LinearInterpolate(A=@n152, B=@n153, Alpha=@n151.B)
    B=VectorParameter() {group=Wrap_ClockWork, parameterName=GearD_Position_Backpack}.RGB
    Alpha=@n154.R
   )
   B=VectorParameter() {group=Wrap_ClockWork, parameterName=GearD_Position Pickaxe, defaultValue=<rgba 0 0 0 1>}.RGB
   Alpha=@n154.G
  )
  NamedRerouteDeclaration(None=ComponentMask(None=LinearInterpolate(A=@n155, B=@n150.RGB, Alpha=@n150.RGB)) {g=true, r=true}) {name=UVs_Offset_D, nodeColor=<rgba 0 0.625 1 1>}
  n156 = ScalarParameter() {defaultValue=0.15, group=Wrap_ClockWork_Uvs, parameterName=Rotation_WeaponCW, sortPriority=0}
  n157 = MaterialFunctionCall(Float4=NamedRerouteUsage()) {materialFunction=&BreakOutFloat4Components}
  n158 = LinearInterpolate(
   A=ScalarParameter() {group=Wrap_ClockWork_Uvs, parameterName=Rotation_VehicleCW, sortPriority=0}
   B=ScalarParameter() {group=Wrap_ClockWork_Uvs, parameterName=Rotation_Vehicle_SideCW, sortPriority=2}
   Alpha=@n157.G
  )
  n159 = ScalarParameter() {defaultValue=1, group=Wrap_ClockWork_Uvs, parameterName=Rotation_MechCW, sortPriority=0}
  n160 = ScalarParameter() {group=Wrap_ClockWork_Uvs, parameterName=Rotation_CharacterCW, sortPriority=2}
  n161 = MaterialFunctionCall(Float2=NamedRerouteUsage()) {materialFunction=&BreakOutFloat2Components}
  n162 = LinearInterpolate(
   A=LinearInterpolate(A=LinearInterpolate(A=@n156, B=@n158, Alpha=@n157.R), B=LinearInterpolate(A=@n159, B=@n160, Alpha=@n157.A), Alpha=@n157.B)
   B=ScalarParameter() {defaultValue=0.25, group=Wrap_ClockWork_Uvs, parameterName=Rotation_BackpackCW, sortPriority=2}
   Alpha=@n161.R
  )
  n163 = ScalarParameter() {group=Wrap_ClockWork_Uvs, parameterName=Rotation_PickaxeCW, sortPriority=2}
  NamedRerouteDeclaration(None=LinearInterpolate(
    A=ScalarParameter() {defaultValue=0.15, group=Wrap_ClockWork_Uvs, parameterName=Rotation---, sortPriority=0}
    B=LinearInterpolate(A=@n162, B=@n163, Alpha=@n161.G)
    Alpha=ScalarParameter() {group=Wrap_ClockWork_Uvs, parameterName="Rotation-Use-States--0=Off", sortPriority=0}
   )) {name=UVs_RotationMSt, nodeColor=<rgba 0 0.7188 1 1>}
  n164 = ScalarParameter() {defaultValue=1, group=Wrap_ClockWork_Uvs, parameterName="CW_Left_Right-Speed  0=On", sortPriority=0}
  NamedRerouteDeclaration(None=Saturate(None=Add(A=@n115, B=@n164))) {name=Left_right, nodeColor=<rgba 1 0.4688 0 1>}
  n165 = MaterialFunctionCall(Color=Abs(None=NamedRerouteUsage())) {materialFunction=&SplitComponents}
  n166 = MaterialFunctionCall(Float4=NamedRerouteUsage()) {materialFunction=&BreakOutFloat4Components}
  n167 = LinearInterpolate(A=@n165.R, B=LinearInterpolate(A=@n165.B, B=@n165.B, Alpha=@n166.G), Alpha=@n166.R)
  NamedRerouteDeclaration(None=MaterialFunctionCall(In=LinearInterpolate(A=@n167, B=@n165.G, Alpha=@n166.B), Contrast=@n97) {materialFunction=&CheapContrast}.Result) {name=Tri_Mask, nodeColor=<rgba 1 0.4688 0 1>}
  NamedRerouteDeclaration(None=@n117.G) {name=IsPickaxe, nodeColor=<rgba 1 0 0.5625 1>}
  n168 = PreSkinnedLocalBounds()
  n169 = Subtract(A=VertexInterpolator(VS=PreSkinnedPosition()).PS, B=@n168.Min)
  n170 = MaterialFunctionCall(Color=Subtract(A=@n168.Max, B=@n168.Max)) {materialFunction=&SplitComponents}
  n171 = Max(A=Max(A=@n170.R, B=@n170.R), B=@n170.B)
  NamedRerouteDeclaration(None=FunctionInput(Preview=Divide(A=@n169, B=@n171)) {bUsePreviewValueAsDefault=true, inputName=Calcated_Bounds, sortPriority=11}) {name=Caclcated_Bounds, nodeColor=<rgba 0 0.9297 1 1>}
  n172 = StaticBool()
  NamedRerouteDeclaration(None=MaterialFunctionCall!(
    Normal_Map
    Offset Rt
    Pivot
    Caculated_Bounds=NamedRerouteUsage()
    Pre_Skinned_Normals=NamedRerouteUsage()
    is_State=NamedRerouteUsage()
    UV_Rotation
    Uvs_Offset
    Uvs_TileRate
    Uvs_Offset_Pivot
    For_Safe_Zones
    Rotation_Comp
    Clamp Y=@n172
    Clamp X=@n172
    Clamp Padding_Comp
    No_Rotation=@n172
   ) {materialFunction=&/MF_Projection_Side_Normal_V2}.UVs No Rotation) {name=Uvs, nodeColor=<rgba 0 0.7422 1 1>}
  NamedRerouteDeclaration(None=Add(
    A=MaterialFunctionCall(X=@n31, Y=@n41, Z=@n61, A=@n56) {materialFunction=&MakeFloat4}.Result
    B=NamedRerouteUsage()
   )) {name="Rotations  ABCD", nodeColor=<rgba 1 0 0.6328 1>}
  ScalarParameter() {group=_UVs, parameterName=Rotation_Weapon, sortPriority=150}
  ScalarParameter() {defaultValue=1, group=_Base, parameterName=Blend_Contrast}
  ScalarParameter() {group=_UVs, parameterName=Rotation_Vehicle, sortPriority=150}
  ScalarParameter() {group=_UVs, parameterName=Rotation_Mech, sortPriority=150}
  VectorParameter() {group=_UVs, parameterName=Offset_Weapon, defaultValue=<rgba 0 0 0 1>}
  VectorParameter() {group=_UVs, parameterName=Offset_Mech, defaultValue=<rgba 0 0 0 1>}
  VectorParameter() {group=_UVs, parameterName=Offset_VehicleSide, defaultValue=<rgba 0 0 0 1>}
  VectorParameter() {group=Wrap_ClockWork, parameterName=GearA_Position}
  VectorParameter() {group=Wrap_ClockWork, parameterName=GearA_Position Weapon, defaultValue=<rgba -0.05 -0.1 0 1>}
 ]}
