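# Material function for the Constellation weapon, vehicle and mech wrap, which replaces emissive and blacks out base color and specular behind a static switch.
# Three star-texture layers, two panning, use virtual-plane UVs with camera-angle parallax.
# Crevasse-masked color and fresnel tints add onto the stars.
# The optional Sun variant layers scrolling hue-shifted noise, detail colors and emissive variation through a warped mask projected from pre-skinned bounds.
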MaterialFunction /Game/Weapons/WeaponSkins/Wraps/MaterialFunctions/MF_ConstellationWrap
graph {nodes [
  n1 = PreSkinnedLocalBounds()
  Add(A=@n1.Min) {constB=0}
  n2 = Subtract(A=ActorPositionWS(), B=CameraPositionWS())
  n3 = ComponentMask(None=@n2) {b=true, r=true}
  n4 = ComponentMask(None=@n2) {g=true, r=true}
  n5 = Multiply(A=DotProduct(A=Normalize(VectorInput=@n4), B=Constant2Vector() {g=1})) {constB=0.5}
  n6 = DotProduct(A=Normalize(VectorInput=@n2), B=Constant3Vector() {constant=<rgba 0 0 1 0>})
  n7 = AppendVector(A=Multiply(A=Multiply(A=@n3) {constB=-1}, B=Add(A=@n5) {constB=0.5}), B=Add(A=Multiply(A=@n6) {constB=0.5}) {constB=0.5})
  n8 = Multiply(
   A=ComponentMask(None=VertexInterpolator(VS=@n7).PS) {g=true, r=true}
   B=ScalarParameter() {defaultValue=0.01, group=Constellation, parameterName=Parallax_Master_Strength}
  ) {constB=0.05}
  n9 = MaterialFunctionCall!(
   Camera _pos
   Axis
   Scale
   Scale_Multplyer
   Object-SafeZone-Virtual-Pivot=NamedRerouteUsage()
   Object-SafeZone-Virtual-CH-Switch=NamedRerouteUsage()
   SinglePlane Normal Axis
   SinglePlane X Axis
   PlaneCenter_distortion
  ) {materialFunction=&/MF_SafeZone_VirtualPlane}
  Add(A=@n8, B=@n9.Uvs)
  Add() {constB=0}
  AppendVector(A=Sign(None=ComponentMask(None=@n2) {r=true}), B=ComponentMask(None=@n2) {b=true})
  n10 = MaterialFunctionCall(Float2=ViewSize()) {materialFunction=&BreakOutFloat2Components}
  n11 = AppendVector(A=Constant() {r=1}, B=Divide(A=@n10.G, B=@n10.G))
  n12 = Multiply(A=ConstantBiasScale(None=ScreenPosition().ViewportUV) {bias=-0.5, scale=2}, B=@n11)
  n13 = ActorPositionWS()
  n14 = Min(A=MaterialFunctionCall(World Position=@n13) {materialFunction=&TransformToClipSpace}.Clip Space XY)
  n15 = Constant()
  n16 = Constant()
  n17 = Multiply(
   A=Multiply(A=Constant2Vector() {g=-0.5, r=0.5}, B=Max(A=@n14) {constB=-1})
   B=If(A=Distance(A=@n13, B=@n15), B=Constant() {r=0.01}, "A > B"=Constant() {r=1}, "A == B"=@n16, "A < B"=@n16)
  )
  n18 = VertexInterpolator(VS=Subtract(A=ConstantBiasScale(None=@n12), B=Multiply(A=@n11, B=@n17)))
  ComponentMask(None=@n18.PS) {g=true, r=true}
  Constant()
  n19 = FunctionInput() {inputName=Material, inputType=FunctionInput_MaterialAttributes, sortPriority=-200}
  n20 = GetMaterialAttributes(None=@n19) {attributeGetTypes [#1, #2, #3]}
  n21 = Constant()
  n22 = VectorParameter() {group=Constellation, parameterName=CS_EmissiveColor, defaultValue=<rgba 0.006 0.9647 0.956 1>}
  n23 = FunctionInput() {inputName="Customization Crevasse (_CM G)", inputType=FunctionInput_Scalar, sortPriority=21, previewValue=<1 0 0 1>}
  n24 = Constant()
  n25 = ScalarParameter() {defaultValue=0.5, group=Constellation, parameterName=NoCrevasseFresnelMaskValue}
  n26 = Abs(None=Add(A=If(A=@n23, B=@n24, "A > B"=@n23, "A == B"=@n25, "A < B"=@n23)) {constB=-0.5})
  n27 = Max(A=Saturate(None=Multiply(A=@n26) {constB=2}), B=ScalarParameter() {defaultValue=0.12, group=Constellation, parameterName=CrevasseMinLevel})
  n28 = FunctionInput() {inputName=IsVehicle, inputType=FunctionInput_Scalar, sortPriority=4}
  n29 = FunctionInput() {inputName=IsMech, inputType=FunctionInput_Scalar, sortPriority=4}
  n30 = ComponentMask(None=Saturate(None=Add(A=@n28, B=@n29))) {r=true}
  n31 = LinearInterpolate(
   A=ScalarParameter() {defaultValue=1, group=Constellation, parameterName=CrevasseFresnelMaskPowWeapons}
   B=ScalarParameter() {defaultValue=1, group=Constellation, parameterName=CrevasseFresnelMaskPowVehicles}
   Alpha=@n30
  )
  n32 = DotProduct(
   A=CameraVectorWS()
   B=StaticSwitchParameter(True=PixelNormalWS(), False=VertexNormalWS()) {group=Constellation, parameterName=FresnelUsesPixelNormals}
  )
  n33 = OneMinus(None=@n32)
  n34 = ScalarParameter() {group=Constellation, parameterName=CS_EmissiveFalloff2}
  n35 = Clamp(None=Power(Base=Subtract(A=@n33, B=@n34) {constB=0.35}, Exp=ScalarParameter() {defaultValue=2, group=Constellation, parameterName=CS_EmissvePower2}))
  n36 = VectorParameter() {group=Constellation, parameterName=CS_EmissiveColor2, defaultValue=<rgba 0 0.1667 0.5573 1>}
  n37 = LinearInterpolate(
   A=ScalarParameter() {defaultValue=0.8, group=Constellation, parameterName=CrevasseFresnelMaskPow2Weapons}
   B=ScalarParameter() {defaultValue=0.8, group=Constellation, parameterName=CrevasseFresnelMaskPow2Vehicles}
   Alpha=@n30
  )
  n38 = Add(A=Multiply(A=@n22.RGB, B=Power(Base=@n27, Exp=@n31)), B=Multiply(A=Multiply(A=@n35, B=@n36.RGB), B=Power(Base=@n27, Exp=@n37)))
  n39 = VectorParameter() {group=Constellation, parameterName=ScreenCoordScaleWeapons, defaultValue=<rgba 10 10 0 0>}
  n40 = VectorParameter() {group=Constellation, parameterName=ScreenCoordScaleVehicles, defaultValue=<rgba 10 10 0 0>}
  n41 = Multiply(A=@n9.Uvs, B=ComponentMask(None=LinearInterpolate(A=@n39.RGB, B=@n40.RGB, Alpha=@n30)) {g=true, r=true})
  n42 = ScalarParameter() {defaultValue=1, group=Constellation, parameterName=Parallax0_X}
  n43 = ScalarParameter() {defaultValue=1, group=Constellation, parameterName=Parallax0_Y}
  n44 = Add(A=@n41, B=Multiply(A=@n8, B=AppendVector(A=@n42, B=@n43)))
  n45 = LinearInterpolate(
   A=ScalarParameter() {defaultValue=0.25, group=Constellation, parameterName=CS_Layer0_Tiling}
   B=ScalarParameter() {defaultValue=0.25, group=Constellation, parameterName=CS_Layer0_Backpack_Tiling}
   Alpha=NamedRerouteUsage()
  )
  n46 = VectorParameter() {group=Constellation, parameterName=CS_Layer_01_PanSpeedAndDirection, defaultValue=<rgba 0.035 -0.25 0 1>}
  n47 = TextureObjectParameter() {
   group=Constellation
   parameterName=StarsTexture
   texture=/Game/Characters/Player/Female/Medium/Bodies/F_MED_Rebirth_Soldier/Skins/Constellation/Textures/T_Constellation_Pattern
  }
  n48 = TextureSample(UVs=Panner(Coordinate=Multiply(A=@n44, B=@n45), Speed=ComponentMask(None=@n46.RGB) {g=true, r=true}), Tex=@n47) {constMipValue=0, texture=/Game/Characters/Player/Female/Medium/Bodies/F_MED_Gift_Wrap/Textures/T_Snowflake_Pattern}
  n49 = VectorParameter() {group=Constellation, parameterName=Layer1_color, defaultValue=<rgba 0.0314 3.8406 5 1>}
  n50 = ScalarParameter() {group=Constellation, parameterName=Parallax1_X}
  n51 = ScalarParameter() {group=Constellation, parameterName=Parallax1_Y}
  n52 = Add(A=@n41, B=Multiply(A=@n8, B=AppendVector(A=@n50, B=@n51))) {constB=0}
  n53 = LinearInterpolate(
   A=ScalarParameter() {defaultValue=0.75, group=Constellation, parameterName=CS_Layer1_Tiling}
   B=ScalarParameter() {defaultValue=0.75, group=Constellation, parameterName=CS_Layer1_Backpack_Tiling}
   Alpha=NamedRerouteUsage()
  )
  n54 = VectorParameter() {group=Constellation, parameterName=CS_Layer_02_PanSpeedAndDirection, defaultValue=<rgba 0.035 -0.25 0 1>}
  n55 = TextureSample(UVs=Panner(Coordinate=Multiply(A=@n52, B=@n53), Speed=ComponentMask(None=@n54.RGB) {g=true, r=true}), Tex=@n47) {texture=&T_Snowflake_Pattern}
  n56 = VectorParameter() {group=Constellation, parameterName=Layer2_color, defaultValue=<rgba 0.0125 1.5363 2 1>}
  n57 = AppendVector(A=ScalarParameter() {group=Constellation, parameterName=Parallax2_X}, B=ScalarParameter() {group=Constellation, parameterName=Parallax2_Y})
  n58 = Multiply(
   A=Add(A=@n41, B=Multiply(A=@n8, B=@n57)) {constB=0}
   B=ScalarParameter() {defaultValue=1.25, group=Constellation, parameterName=CS_Layer2_Tiling}
  )
  n59 = Max(
   A=Max(A=Multiply(A=@n48.G, B=@n49.RGB), B=Multiply(A=@n55.R, B=@n56.RGB) {constB=0.5})
   B=Multiply(
    A=TextureSample(UVs=@n58, Tex=@n47) {texture=&T_Snowflake_Pattern}.B
    B=VectorParameter() {group=Constellation, parameterName=Layer3_color, defaultValue=<rgba 0.0063 1 0.991 1>}.RGB
   ) {constB=0.25}
  )
  n60 = ScalarParameter() {group=ConstellationSun, parameterName=PatternEmissiveStrength}
  n61 = ScalarParameter() {group=ConstellationSun, parameterName=NoiseScrollingSpeedU}
  n62 = ScalarParameter() {group=ConstellationSun, parameterName=NoiseScrollingSpeedV}
  n63 = Add(
   A=TextureCoordinate()
   B=Multiply(A=Time(), B=MaterialFunctionCall(X=@n61, Y=@n62) {materialFunction=&MakeFloat2}.Result)
  )
  n64 = ScalarParameter() {group=ConstellationSun, parameterName=NoiseRotation}
  n65 = TextureSampleParameter2D(UVs=Multiply(
    A=MaterialFunctionCall(UVs=@n63, "Rotation Angle (0-1)"=@n64) {materialFunction=&CustomRotator}.Rotated Values
    B=ScalarParameter() {defaultValue=1, group=ConstellationSun, parameterName=NoiseTiling}
   )) {group=ConstellationSun, parameterName=NoiseTexture, texture=&T_BlacksRGB}
  n66 = Power(
   Base=Divide(A=MaterialFunctionCall(f3=@n65.RGB) {materialFunction=&AddComponents}.f3) {constB=3}
   Exp=ScalarParameter() {defaultValue=1, group=ConstellationSun, parameterName=NoisePow}
  )
  n67 = VectorParameter() {group=ConstellationSun, parameterName=ConstellationSurfaceDetailsColorA_NoiseColor, defaultValue=<rgba 0.974 0.224 0.0287 0>}
  n68 = MaterialFunctionCall(Hue Shift Percentage=Multiply(A=@n66, B=ScalarParameter() {group=ConstellationSun, parameterName=NoiseHue}), Texture=@n67.RGB) {materialFunction=&HueShift}
  n69 = TextureCoordinate()
  n70 = LinearInterpolate(
   A=ScalarParameter() {defaultValue=1, group=ConstellationSun, parameterName=CS-WarpUVS-01-TileRate Weapon, sortPriority=53}
   B=ScalarParameter() {defaultValue=1, group=ConstellationSun, parameterName=CS-WarpUVS-01-TileRate Vehicle, sortPriority=54}
   Alpha=@n28
  )
  n71 = Time() {period=1}
  n72 = VectorParameter() {group=ConstellationSun, parameterName=CS-WarpUVS-01-Anitmated Weapon, sortPriority=57, defaultValue=<rgba 0 0 0 1>}
  n73 = VectorParameter() {group=ConstellationSun, parameterName=CS-WarpUVS-01-Anitmated Vehicle, sortPriority=57, defaultValue=<rgba 0 0 0 1>}
  n74 = ComponentMask(None=LinearInterpolate(A=@n72.RGB, B=@n73.RGB, Alpha=@n28)) {g=true, r=true}
  n75 = TextureSampleParameter2D(UVs=Add(A=Multiply(A=@n69, B=@n70), B=Multiply(A=@n71, B=@n74))) {group=ConstellationSun, parameterName=CS-WarpUVS-01-Texture, sortPriority=52, texture=&T_BlacksRGB}
  n76 = ChannelMaskParameter(None=@n75.RGB) {group=ConstellationSun, parameterName=CS-WarpUVS-01_Channel, sortPriority=51}
  n77 = LinearInterpolate(
   A=ScalarParameter() {group=ConstellationSun, parameterName=CS-WarpUVS-01-Amount Weapon, sortPriority=55}
   B=ScalarParameter() {group=ConstellationSun, parameterName=CS-WarpUVS-01-Amount_Vehicle, sortPriority=56}
   Alpha=@n28
  )
  n78 = VectorParameter() {group=ConstellationSun, parameterName="CS_Scale_Weapon_01 A=uniform", sortPriority=30, defaultValue=<rgba 1 1 1 1>}
  n79 = VectorParameter() {group=ConstellationSun, parameterName="CS_Scale_Vehicle_01 A=uniform", sortPriority=31, defaultValue=<rgba 1 1 1 1>}
  n80 = Multiply(A=@n79.RGB, B=@n79.RGB)
  n81 = VectorParameter() {group=ConstellationSun, parameterName="CS_Scale_Vehicle_Side_01 A=uniform", sortPriority=31, defaultValue=<rgba 1 1 1 1>}
  n82 = Multiply(A=@n81.RGB, B=@n81.RGB)
  n83 = FunctionInput() {bUsePreviewValueAsDefault=true, inputName=Side Project Vehicle, inputType=FunctionInput_Scalar}
  n84 = VectorParameter() {group=ConstellationSun, parameterName="CS_Scale_Mech_01 A=uniform", defaultValue=<rgba 1 1 1 1>}
  n85 = LinearInterpolate(
   A=LinearInterpolate(A=Multiply(A=@n78.RGB, B=@n78.RGB), B=LinearInterpolate(A=@n80, B=@n82, Alpha=@n83), Alpha=@n28)
   B=Multiply(A=@n84.RGB, B=@n84.RGB)
   Alpha=@n29
  )
  n86 = PreSkinnedPosition()
  n87 = PreSkinnedLocalBounds()
  n88 = Divide(A=Subtract(A=VertexInterpolator(VS=@n86).PS, B=@n87.Min), B=Subtract(A=@n87.Max, B=@n87.Max))
  n89 = VectorParameter() {group=ConstellationSun, parameterName=CS_TextureOffset_Weapon, sortPriority=15, defaultValue=<rgba 0 0 0 1>}
  n90 = VectorParameter() {group=ConstellationSun, parameterName=CS_TextureOffset_Vehicle, sortPriority=15, defaultValue=<rgba 0 0 0 1>}
  n91 = VectorParameter() {group=ConstellationSun, parameterName=CS_TextureOffset_Vehicle_Side, sortPriority=15, defaultValue=<rgba 0 0 0 1>}
  n92 = LinearInterpolate(A=@n90.RGB, B=@n91.RGB, Alpha=@n83)
  n93 = VectorParameter() {group=ConstellationSun, parameterName=CS_TextureOffset_Mech, sortPriority=15, defaultValue=<rgba 0 0 0 1>}
  n94 = Add(
   A=Multiply(
    A=FunctionInput(Preview=@n88) {bUsePreviewValueAsDefault=true, inputName=Wrap Bounds, sortPriority=12}
    B=Constant3Vector() {constant=<rgba 1 1 -1 0>}
   )
   B=LinearInterpolate(A=LinearInterpolate(A=@n89.RGB, B=@n92, Alpha=@n28), B=@n93.RGB, Alpha=@n29)
  )
  n95 = MaterialFunctionCall(Color=Add(A=Multiply(A=@n76.R, B=@n77), B=Multiply(A=@n94, B=@n85))) {materialFunction=&SplitComponents}
  n96 = AppendVector(A=@n95.G, B=@n95.G)
  n97 = LinearInterpolate(A=AppendVector(A=@n95.R, B=@n95.R), B=AppendVector(A=@n95.R, B=@n95.R), Alpha=@n83)
  n98 = OneMinus(None=@n95.B)
  n99 = TextureSampleParameter2D(UVs=LinearInterpolate(A=LinearInterpolate(A=@n96, B=@n97, Alpha=@n28), B=AppendVector(A=@n98, B=@n95.R), Alpha=@n29)) {group=ConstellationSun, parameterName=ConstellationSurfaceDetailsMask, texture=&T_BlacksRGB}
  n100 = VectorParameter() {group=ConstellationSun, parameterName=ConstellationSurfaceDetailsColorG, defaultValue=<rgba 0.974 0.224 0.0287 0>}
  n101 = LinearInterpolate(A=LinearInterpolate(A=Multiply(A=@n59, B=@n60), B=Multiply(A=@n66, B=@n68.Result), Alpha=@n99.A), B=@n100.RGB, Alpha=@n99.G)
  n102 = VectorParameter() {group=ConstellationSun, parameterName=ConstellationSurfaceDetailsColorR, defaultValue=<rgba 0.974 0.224 0.0287 0>}
  n103 = Power(
   Base=@n99.R
   Exp=ScalarParameter() {defaultValue=1, group=ConstellationSun, parameterName=ConstellationSurfaceDetailsColorR_MaskPow, sortPriority=79}
  )
  n104 = VectorParameter() {group=ConstellationSun, parameterName=ConstellationSurfaceDetailsColorB, defaultValue=<rgba 0.974 0.224 0.0287 0>}
  n105 = Add(A=@n38, B=LinearInterpolate(A=LinearInterpolate(A=@n101, B=@n102.RGB, Alpha=@n103), B=@n104.RGB, Alpha=@n99.B))
  n106 = TextureCoordinate()
  n107 = ScalarParameter() {defaultValue=1, group=ConstellationSun, parameterName=CS_EmissiveVariation_TileRate, sortPriority=201}
  n108 = Time() {period=1}
  n109 = VectorParameter() {group=ConstellationSun, parameterName=CS_EmissiveVariation_Animation, sortPriority=201, defaultValue=<rgba 0 0 0 1>}
  n110 = ComponentMask(None=@n109.RGB) {g=true, r=true}
  n111 = TextureSampleParameter2D(UVs=Add(A=Multiply(A=@n106, B=@n107), B=Multiply(A=@n108, B=@n110))) {group=ConstellationSun, parameterName=CS-WarpUVS-01-Texture, sortPriority=52, texture=&T_BlacksRGB}
  n112 = ChannelMaskParameter(None=@n111.RGBA) {group=ConstellationSun, parameterName=CS-EmissiveVariation_ChannelMask, sortPriority=200}
  n113 = VectorParameter() {group=ConstellationSun, parameterName=CS_EmissiveVariationColor, sortPriority=200, defaultValue=<rgba 0.974 0.224 0.0287 0>}
  n114 = Add(
   A=Power(
    Base=Multiply(A=@n112.R, B=@n113.RGB)
    Exp=ScalarParameter() {defaultValue=1, group=ConstellationSun, parameterName=CS_EmissiveVariation_MaskPow, sortPriority=202}
   )
   B=@n105
  )
  n115 = ScalarParameter() {desc="1 = on, 0 = off", group=ConstellationSun, parameterName=CS_UseEmissiveVariation, sortPriority=199}
  n116 = Subtract(A=Saturate(None=@n32)) {constB=0.1}
  n117 = Clamp(None=Multiply(A=Power(Base=@n116)) {constB=2})
  n118 = StaticSwitchParameter(
   True=LinearInterpolate(A=@n105, B=LinearInterpolate(A=@n105, B=@n114, Alpha=@n115), Alpha=Add(A=@n99.R, B=@n99.R) {desc=Emissive Variation Mask})
   False=Add(A=Multiply(A=@n59, B=@n117), B=@n38)
  ) {
   desc="Screen Space Projection Base (3 parallax layers) - Side Projection texture base for masked sections - Topside Emissive RGB parallax channels - Reactive (Glow on Fire)"
   group=Customization
   parameterName=Constellation Sun Wrap
  }
  FunctionOutput(None=StaticSwitchParameter(
    True=SetMaterialAttributes!(MaterialAttributes=@n20.MaterialAttributes, BaseColor=@n21, EmissiveColor=@n118, Specular=@n21) {attributeSetTypes [#1, #2, #4]}
    False=@n19
   ) {desc="Screen Space Projection Base (3 parallax layers)", group=Customization_Wrap, parameterName=Constellation Wrap})
  n119 = Multiply(A=Subtract(A=@n1.Max, B=@n1.Max)) {constB=0.5}
  n120 = ObjectPositionWS()
  LinearInterpolate(A=TransformPosition(None=Add(A=@n1.Min, B=@n119)) {transformType=TRANSFORMPOSSOURCE_World}, B=@n120)
  n121 = Normalize(VectorInput=Subtract(A=@n120, B=CameraPositionWS()))
  n122 = Constant3Vector() {constant=<rgba 0 0 1 1>}
  n123 = MaterialFunctionCall(Float3=PreSkinnedLocalBounds().Half Extents) {materialFunction=&BreakOutFloat3Components}
  n124 = Multiply(
   A=MaterialFunctionCall() {materialFunction=/Engine/Functions/Engine_MaterialFunctions02/WorldPositionOffset/ObjectScale}.Scale X
   B=Max(A=Max(A=@n123.R, B=@n123.R), B=@n123.B)
  )
  MaterialFunctionCall!(Plane Normal Axis=@n121, Plane X Axis=Normalize(VectorInput=CrossProduct(A=@n121, B=@n122)), Plane Center=@n120, UV Scale=@n124) {materialFunction=/Game/Athena/Environments/Nexus/Materials/VirtualPlaneCoordinatesAbs}
  n125 = VectorParameter() {parameterName=Axis}
  n126 = ScalarParameter() {defaultValue=1, group=Constellation, parameterName=ScreenCoordScale_Weapons}
  n127 = ScalarParameter() {defaultValue=1, group=Constellation, parameterName=ScreenCoordScale_Vehicles}
  MaterialFunctionCall(
   Plane Normal Axis=@n121
   Plane X Axis=Normalize(VectorInput=CrossProduct(A=@n121, B=@n125.RGB))
   Plane Center=@n120
   UV Scale=Multiply(A=@n124, B=LinearInterpolate(A=@n126, B=@n127))
  ) {materialFunction=&VirtualPlaneCoordinatesAbs}
  n128 = Distance(A=ComponentMask(None=ActorPositionWS()) {b=true, g=true, r=true}, B=Constant3Vector())
  n129 = OneMinus(None=Round(None=Saturate(None=@n128)))
  n130 = LinearInterpolate(A=ObjectRadius(), B=Constant() {r=100}, Alpha=Saturate(None=Multiply(A=@n129, B=@n30) {constA=1}))
  n131 = ActorPositionWS()
  n132 = Max(
   A=ComponentMask(None=TransformPosition(None=@n131) {transformSourceType=TRANSFORMPOSSOURCE_World, transformType=TRANSFORMPOSSOURCE_View}) {b=true}
   B=Constant() {r=100}
  )
  MaterialFunctionCall(
   UVs=@n18.PS
   Texture Scale=Multiply(A=Divide(A=@n130, B=@n132), B=ScalarParameter() {defaultValue=1, group=Constellation, parameterName=CS_UVRadialScaleZero})
  ) {materialFunction=&ScaleUVsByCenter}
  Multiply(A=CameraPositionWS(), B=Constant3Vector() {constant=<rgba -1 -1 -1 0>})
  n133 = Subtract(A=@n33, B=ScalarParameter() {group=Constellation, parameterName=CS_EmissiveFalloff}) {constB=0.35}
  n134 = ScalarParameter() {defaultValue=2, group=Constellation, parameterName=CS_EmissvePower}
  Multiply(A=Clamp(None=Power(Base=@n133, Exp=@n134)), B=@n22.RGB)
  n135 = FunctionInput() {bUsePreviewValueAsDefault=true, inputName=IsStatsB, inputType=FunctionInput_Vector2, sortPriority=78}
  NamedRerouteDeclaration(None=ComponentMask(None=NamedRerouteDeclaration(None=@n135) {name=Is_StatsB, nodeColor=<rgba 0 0.9253 1 1>}) {r=true}) {name=IsBackpack, nodeColor=<rgba 0.2656 0 1 1>}
  n136 = FunctionInput(Preview=Constant4Vector() {constant=<rgba 0 0 5 -1>}.RGBA) {bUsePreviewValueAsDefault=true, inputName=Object-SafeZone-Virtual-Pivot, inputType=FunctionInput_Vector4, sortPriority=105}
  NamedRerouteDeclaration(None=@n136) {name=SafeZone VP, nodeColor=<rgba 0.5 1 0 1>}
  n137 = FunctionInput(Preview=Constant()) {bUsePreviewValueAsDefault=true, inputName=Object-SafeZone-VP-Character-PreView-Switch, inputType=FunctionInput_Scalar, sortPriority=105}
  NamedRerouteDeclaration(None=@n137) {name=SafeZone Character-Switch, nodeColor=<rgba 1 0.9375 0 1>}
  Normalize(VectorInput=@n3)
  ObjectRadius()
 ]}
