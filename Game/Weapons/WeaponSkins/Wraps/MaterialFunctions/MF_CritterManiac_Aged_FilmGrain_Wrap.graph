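# Weapon wrap material function that gives Critter Maniac "Aged" skins an old film look behind a static switch, optionally limited by a customization mask.
# Builds FOV-scaled, camera-facing virtual-plane UVs for animated static, scratches, hairs and vertical lines, using position-offset timing and time-of-day-scaled static.
# Adds grain layers to emissive, then desaturates it.
# Overrides roughness and darkens base color with a flickering view-angle falloff, except on mobile.
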MaterialFunction /Game/Weapons/WeaponSkins/Wraps/MaterialFunctions/MF_CritterManiac_Aged_FilmGrain_Wrap
graph {nodes [
  n1 = GetMaterialAttributes(None=NamedRerouteUsage()) {attributeGetTypes [#1, #2]}
  n2 = NamedRerouteUsage()
  n3 = Constant() {r=1}
  n4 = NamedRerouteUsage()
  n5 = Add(A=ComponentMask(None=@n4) {r=true}, B=ComponentMask(None=@n4) {g=true})
  n6 = ComponentMask(None=NamedRerouteUsage()) {b=true}
  n7 = Saturate(None=Add(A=Add(A=@n5, B=@n6), B=NamedRerouteUsage()))
  n8 = SetMaterialAttributes!(
   MaterialAttributes=@n1.MaterialAttributes
   BaseColor=Multiply(A=@n1.BaseColor, B=ShadingPathSwitch(Default=@n2, Deferred=@n2, Mobile=@n3) {desc=Migrated from FeatureLevelSwitch})
   EmissiveColor=Desaturation(None=Add(A=@n1.EmissiveColor, B=@n7), Fraction=ScalarParameter() {defaultValue=1, group=CM_Aged, parameterName=CM_Aged_EmissiveDesaturation})
   Roughness=ScalarParameter() {defaultValue=1, group=CM_Aged, parameterName=CM_Aged_Roughness}
  ) {attributeSetTypes [#1, #2, #3]}
  n9 = FunctionInput() {bUsePreviewValueAsDefault=true, inputName="Customization Mask (_CM B)", inputType=FunctionInput_Scalar, sortPriority=24}
  n10 = ScalarParameter() {group=CM_Aged, parameterName=CM_Aged_Remove_Customization Mask, sortPriority=-10}
  n11 = BlendMaterialAttributes(A=NamedRerouteUsage(), B=@n8, Alpha=Saturate(None=Add(A=@n9, B=@n10)))
  FunctionOutput(None=StaticSwitchParameter(True=StaticSwitchParameter(True=@n11, False=@n8) {group=CM_Aged, parameterName=CM_Aged_UseCustomizationMask}, False=NamedRerouteUsage()) {group=Customization, parameterName="Critter_ManiacAged\\FilmGrain"})
  n12 = Constant3Vector() {constant=<rgba 1 1 0 0>}
  n13 = MaterialFunctionCall() {materialFunction=&CameraDirectionVector}
  n14 = Multiply(A=Constant() {desc=Needs at least a small amount of this or it will break, r=0.25}, B=Normalize(VectorInput=Multiply(A=@n12, B=@n13.Result)))
  n15 = Multiply(A=@n13.Result, B=Constant() {r=25})
  n16 = Add(A=ObjectPositionWS(), B=Constant3Vector() {constant=<rgba 0 0 40 1>})
  n17 = ScalarParameter() {defaultValue=0.8, group=CM_Aged, parameterName=CM_Aged_Scale-LowFOV}
  n18 = ScalarParameter() {defaultValue=1.25, group=CM_Aged, parameterName=CM_Aged_Scale-HighFOV}
  n19 = Saturate(None=MaterialFunctionCall(
    Input=MaterialFunctionCall() {materialFunction=/Engine/Functions/Engine_MaterialFunctions02/Utility/FOV}."FOV "
    Input Low=Constant() {r=0.414}
    Input High=Constant() {r=1.091}
    Target Low=Constant()
    Target High=Constant() {r=1}
   ) {materialFunction=&RemapValueRange}.Result)
  NamedRerouteDeclaration(None=MaterialFunctionCall(Plane Normal Axis=Add(A=@n14, B=@n13.Result), Plane Center=Add(A=@n15, B=@n16), UV Scale=LinearInterpolate(A=@n17, B=@n18, Alpha=@n19)) {materialFunction=&VirtualPlaneCoordinates}.UVs) {name=ScreenAlignedUVs, nodeColor=<rgba 0 0.8125 1 1>}
  NamedRerouteDeclaration(None=FunctionInput() {inputName=Material, inputType=FunctionInput_MaterialAttributes}) {name=MaterialAttributes, nodeColor=<rgba 0.125 1 0 1>}
  n20 = ScalarParameter() {defaultValue=0.9, group=CM_Aged, parameterName=CM_Aged_Noise-OffsetUV}
  n21 = Time()
  n22 = Floor(None=Multiply(A=ScalarParameter() {defaultValue=16, group=CM_Aged, parameterName=CM_Aged_Noise-NumberOffsets}, B=Frac(None=@n21)))
  n23 = ScalarParameter() {defaultValue=0.03, group=CM_Aged, parameterName=CM_Aged_Noise-Scale}
  n24 = NamedRerouteUsage()
  n25 = TextureSample(UVs=Add(A=Multiply(A=@n20, B=@n22), B=Multiply(A=@n23, B=@n24))) {
   samplerSource=SSM_Wrap_WorldGroupSettings
   samplerType=SAMPLERTYPE_LinearColor
   texture=/Game/Global/Textures/Default/Noise/SparseGaussianNoise256
  }
  n26 = Multiply(A=OneMinus(None=@n25.R), B=ScalarParameter() {defaultValue=0.015, group=CM_Aged, parameterName=CM_Aged_NoiseAmt})
  n27 = VectorParameter() {group=CM_Aged, parameterName=CM_Aged_Static-Color, defaultValue=<rgba 1 1 1 1>}
  n28 = Constant() {r=0.5}
  n29 = Constant() {r=1}
  NamedRerouteDeclaration(None=Multiply(
    A=Multiply(A=@n26, B=@n27.RGB)
    B=MaterialFunctionCall(Value_Morning=@n28, Value_Day=@n29, Value_Evening=@n28, Value_Night=@n28) {materialFunction=&Scalar_Time_of_Day}.Result
   )) {name=InteriorStaticNoise, nodeColor=<rgba 0.1016 1 0 1>}
  n30 = Multiply(A=NamedRerouteUsage()) {constB=0.01}
  n31 = MaterialFunctionCall(Vector 3=ObjectPositionWS()) {materialFunction=/Engine/Functions/Engine_MaterialFunctions02/Utility/VectorLength}
  n32 = Add(
   A=Multiply(A=Time()) {constB=0.25}
   B=NamedRerouteDeclaration(None=Divide(A=@n31.V3 Length) {constB=250}) {name=PositionBasedOffset, nodeColor=<rgba 0.7578 1 0 1>}
  )
  n33 = TextureSample(UVs=AppendVector(A=Multiply(A=@n32) {constB=0.333}, B=Multiply(A=@n32) {constB=0.777})) {
   samplerSource=SSM_Wrap_WorldGroupSettings
   samplerType=SAMPLERTYPE_LinearColor
   texture=&SparseGaussianNoise256
  }
  n34 = AppendVector(A=@n33.R, B=@n33.R)
  n35 = TextureSample(UVs=Frac(None=Add(A=@n30, B=@n34))) {
   samplerSource=SSM_Wrap_WorldGroupSettings
   samplerType=SAMPLERTYPE_Masks
   texture=/Game/Characters/Player/Male/Medium/Bodies/M_MED_Critter_Frenzy/Textures/T_FilmGrain_Noise
  }
  NamedRerouteDeclaration(None=@n35.RGB) {name=ScratchesAndHairs, nodeColor=<rgba 0.3359 1 0 1>}
  n36 = Multiply(A=NamedRerouteUsage()) {constB=0.005}
  n37 = NamedRerouteUsage()
  n38 = Time()
  n39 = Add(A=Add(A=Multiply(A=@n37) {constB=0.25}, B=Multiply(A=@n38) {constB=0.25}), B=NamedRerouteUsage())
  n40 = AppendVector(A=Multiply(A=Sine(None=@n39)) {constB=0.025}, B=Multiply(A=Time()) {constB=4})
  n41 = TextureSample(UVs=Frac(None=Subtract(A=@n36, B=@n40))) {
   samplerSource=SSM_Wrap_WorldGroupSettings
   samplerType=SAMPLERTYPE_Masks
   texture=&T_FilmGrain_Noise
  }
  n42 = Time()
  n43 = Abs(None=Sine(None=Multiply(A=@n42) {constB=0.5}))
  NamedRerouteDeclaration(None=Multiply(A=@n41.B, B=LinearInterpolate(Alpha=@n43) {constA=0.1})) {name=VerticalLines, nodeColor=<rgba 1 0.9141 0 1>}
  n44 = Add(A=NamedRerouteUsage(), B=Multiply(A=Time(), B=ScalarParameter() {defaultValue=4, group=CM_Aged, parameterName=CM_Aged_FlickerSpeed}))
  n45 = Multiply(A=@n44) {constB=0.966}
  n46 = Multiply(A=@n44) {constB=0.125}
  n47 = Add(A=Add(A=Sine(None=@n45), B=Sine(None=@n46)), B=Sine(None=Multiply(A=@n44) {constB=0.033}))
  NamedRerouteDeclaration(None=Saturate(None=Abs(None=@n47))) {name=Flicker, nodeColor=<rgba 0.7109 0 1 1>}
  n48 = CameraVectorWS()
  n49 = PixelNormalWS()
  n50 = ScalarParameter() {defaultValue=2, group=CM_Aged, parameterName=CM_Aged_ColorFalloff_Exp}
  n51 = Multiply(A=@n50) {constB=0.92}
  n52 = NamedRerouteUsage()
  n53 = Saturate(None=Power(Base=DotProduct(A=@n48, B=@n49), Exp=LinearInterpolate(A=@n51, B=@n50, Alpha=@n52) {constA=1}))
  NamedRerouteDeclaration(None=LinearInterpolate(
    A=ScalarParameter() {defaultValue=1, group=CM_Aged, parameterName=CM_Aged_ColorFalloff_High}
    B=ScalarParameter() {defaultValue=0.02, group=CM_Aged, parameterName=CM_Aged_ColorFalloff_Low}
    Alpha=OneMinus(None=@n53)
   )) {name=BaseColor_Falloff, nodeColor=<rgba 1 0.6094 0 1>}
  QualitySwitch()
  SetMaterialAttributes!(MaterialAttributes=@n8, ShadingModel=ShadingModel()) {attributeSetTypes [#4]}
  VertexNormalWS()
 ]}
