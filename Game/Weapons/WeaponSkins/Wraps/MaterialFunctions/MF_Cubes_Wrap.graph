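# Material function for the Cubes wrap skin on weapons, vehicles and mechs, blended into incoming material attributes by customization mask behind the Wrap_Cubes static switch.
# Two top/side-projected layers with per-platform UV transforms blend base color; the first also drives normal, roughness, specular and metallic.
# A mask-driven sine/cosine offset wobbles the first layer's projection.
# Emissive combines pulsing and height-lerped gradient glow, brightening on fire via Blueprint-set game time, then decaying.
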MaterialFunction /Game/Weapons/WeaponSkins/Wraps/MaterialFunctions/MF_Cubes_Wrap
graph {nodes [
  n1 = Sine(None=Multiply(A=Time(), B=ScalarParameter() {defaultValue=0.1, group=Cubes Wrap Emissive, parameterName=Cubes-Emissive01-BrightnessTime, sortPriority=-10}))
  Clamp(
   None=Divide(A=Add(A=@n1))
   Min=ScalarParameter() {group=Cubes Wrap Emissive, parameterName=Cubes-Emissive01-MinBrightness, sortPriority=-9}
   Max=ScalarParameter() {defaultValue=1, group=Cubes Wrap Emissive, parameterName=Cubes-Emissive01-MaxBrightness, sortPriority=-8}
  )
  n2 = FunctionInput() {inputName=Material, inputType=FunctionInput_MaterialAttributes, sortPriority=-10}
  n3 = GetMaterialAttributes(None=@n2) {attributeGetTypes [#1, #2, #3]}
  n4 = SetMaterialAttributes!(
   MaterialAttributes=NamedRerouteUsage()
   BaseColor=LinearInterpolate(A=NamedRerouteUsage(), B=NamedRerouteUsage(), Alpha=NamedRerouteUsage())
   Roughness=NamedRerouteUsage()
   Normal=NamedRerouteUsage()
   Specular=NamedRerouteUsage()
   Metallic=NamedRerouteUsage()
   EmissiveColor=NamedRerouteUsage()
  ) {attributeSetTypes [
    #1
    #2
    #3
    #4
    #5
    #6
   ]}
  n5 = StaticSwitchParameter(
   True=BlendMaterialAttributes(
    A=@n3.MaterialAttributes
    B=@n4
    Alpha=FunctionInput() {bUsePreviewValueAsDefault=true, inputName="Customization Mask (_CM B)", inputType=FunctionInput_Scalar, sortPriority=24}
   ) {vertexAttributeBlendType=UseA}
   False=@n2
  ) {
   desc="TopSide Projection Base - Ambient emissive  - Reactive -Topside Emissive with Texture and Ambient becomes brighter Height Lerp on Fire"
   group=Customization_Wrap
   parameterName=Wrap_Cubes
  }
  FunctionOutput(None=@n5)
  n6 = NamedRerouteUsage()
  n7 = NamedRerouteUsage()
  n8 = TextureObjectParameter() {group=Cubes Wrap, parameterName=Cubes_Emissive_01, sortPriority=11, texture=&T_BlacksRGB}
  n9 = LinearInterpolate(
   A=ScalarParameter() {defaultValue=1, group=Cubes Wrap Emissive, parameterName=Cubes_Blend_Contrast, sortPriority=-1}
   B=ScalarParameter() {defaultValue=2, group=Cubes Wrap, parameterName=Cubes_Blend_Contrast_Mech, sortPriority=-1}
   Alpha=NamedRerouteUsage()
  )
  n10 = ScalarParameter() {group=Cubes Wrap, parameterName=Cubes_Fill_alternate_Axis_Opacity 01, sortPriority=2}
  n11 = VectorParameter() {group=Cubes Wrap, parameterName=Cubes_Fill_alternate_Axis_Color_Mask_01, sortPriority=21, defaultValue=<rgba 0 0 0 1>}
  n12 = AppendVector(A=@n11.RGB, B=@n11.RGB)
  n13 = TextureObjectParameter() {
   group=Cubes Wrap
   parameterName=Cubes_WobbleMask
   samplerType=SAMPLERTYPE_Masks
   sortPriority=600
   texture=&T_BlackMask
  }
  n14 = Constant()
  n15 = VectorParameter() {group=Cubes Wrap UVs, parameterName=Cubes-Rotation-Weapon-01---------------------, sortPriority=1, defaultValue=<rgba 0 0 -0.25 1>}
  n16 = VectorParameter() {group=Cubes Wrap UVs, parameterName=Cubes_Rotation_Vehicle_01, sortPriority=2, defaultValue=<rgba 0 0 0 1>}
  n17 = VectorParameter() {
   desc=must activate TS_Vehicle_Side_01 UVs
   group=Cubes Wrap UVs
   parameterName=Cubes_Rotation_Vehicle_Side_01
   sortPriority=2
   defaultValue=<rgba 0 0 0 1>
  }
  n18 = Multiply(A=NamedRerouteUsage(), B=ScalarParameter() {group=Cubes Wrap UVs, parameterName="Cubes_Vehicle_Side_01 UVs 0=Off 1=Active"})
  n19 = VectorParameter() {group=Cubes Wrap UVs, parameterName=Cubes_Rotation_Mech_01, sortPriority=3, defaultValue=<rgba 0.25 0.25 0 1>}
  n20 = LinearInterpolate(
   A=LinearInterpolate(A=@n15.RGB, B=LinearInterpolate(A=@n16.RGB, B=@n17.RGB, Alpha=@n18), Alpha=NamedRerouteUsage())
   B=@n19.RGB
   Alpha=NamedRerouteUsage()
  )
  n21 = NamedRerouteUsage()
  n22 = VectorParameter() {group=Cubes Wrap UVs, parameterName=Cubes_Offset_Weapon_01---------------------, sortPriority=10, defaultValue=<rgba 0 0 0 1>}
  n23 = VectorParameter() {group=Cubes Wrap UVs, parameterName=Cubes_Offset_Vehicle_01, sortPriority=11, defaultValue=<rgba 0 0 0 1>}
  n24 = VectorParameter() {
   desc=must activate TS_Vehicle_Side_01 UVs
   group=Cubes Wrap UVs
   parameterName=Cubes_Offset_Vehicle_Side_01
   sortPriority=11
   defaultValue=<rgba 0 0 0 1>
  }
  n25 = VectorParameter() {group=Cubes Wrap UVs, parameterName=Cubes_Offset_Mech_01, sortPriority=12, defaultValue=<rgba 0 0 0 1>}
  n26 = LinearInterpolate(
   A=LinearInterpolate(A=@n22.RGB, B=LinearInterpolate(A=@n23.RGB, B=@n24.RGB, Alpha=@n18), Alpha=NamedRerouteUsage())
   B=@n25.RGB
   Alpha=NamedRerouteUsage()
  )
  n27 = Time()
  n28 = VectorParameter() {group=Cubes Wrap UVs, parameterName=Cubes_Panning_Weapon_01---------------------, sortPriority=40, defaultValue=<rgba 0 0 0 1>}
  n29 = VectorParameter() {group=Cubes Wrap UVs, parameterName=Cubes_Panning_Vehicle_01, sortPriority=41, defaultValue=<rgba 0 0 0 1>}
  n30 = VectorParameter() {
   desc=must activate TS_Vehicle_Side_01 UVs
   group=Cubes Wrap UVs
   parameterName=Cubes_Panning_Vehicle_Side_01
   sortPriority=41
   defaultValue=<rgba 0 0 0 1>
  }
  n31 = VectorParameter() {group=Cubes Wrap UVs, parameterName=Cubes_Panning_Mech_01, sortPriority=42, defaultValue=<rgba 0 0 0 1>}
  n32 = LinearInterpolate(
   A=LinearInterpolate(A=@n28.RGB, B=LinearInterpolate(A=@n29.RGB, B=@n30.RGB, Alpha=@n18), Alpha=NamedRerouteUsage())
   B=@n31.RGB
   Alpha=NamedRerouteUsage()
  )
  n33 = VectorParameter() {group=Cubes Wrap UVs, parameterName="Cubes_Scale_Weapon_01 A=uniform", sortPriority=30, defaultValue=<rgba 1 1 1 1>}
  n34 = Multiply(A=@n33.RGB, B=@n33.RGB)
  n35 = VectorParameter() {group=Cubes Wrap UVs, parameterName="Cubes_Scale_Vehicle_01 A=uniform", sortPriority=31, defaultValue=<rgba 1 1 1 1>}
  n36 = VectorParameter() {
   desc=must activate TS_Vehicle_Side_01 UVs
   group=Cubes Wrap UVs
   parameterName="Cubes_Scale_Vehicle_Side_01 A=uniform"
   sortPriority=31
   defaultValue=<rgba 1 1 1 1>
  }
  n37 = LinearInterpolate(A=Multiply(A=@n35.RGB, B=@n35.RGB), B=Multiply(A=@n36.RGB, B=@n36.RGB), Alpha=@n18)
  n38 = NamedRerouteUsage()
  n39 = VectorParameter() {group=Cubes Wrap UVs, parameterName="Cubes_Scale_Mech_01 A=uniform", defaultValue=<rgba 1 1 1 1>}
  n40 = Multiply(
   A=Add(A=Add(A=@n21, B=@n26), B=Multiply(A=@n27, B=@n32))
   B=LinearInterpolate(A=LinearInterpolate(A=@n34, B=@n37, Alpha=@n38), B=Multiply(A=@n39.RGB, B=@n39.RGB), Alpha=NamedRerouteUsage())
  )
  n41 = MaterialFunctionCall!(
   IsCharacter=NamedRerouteUsage()
   IsStatsB=NamedRerouteUsage()
   Texture=@n13
   Rotation_Axis=@n20
   Caculated_Bounds=@n40
   Pre_Skinned_Normals=NamedRerouteUsage()
   Blend_Contrast=LinearInterpolate(
    A=ScalarParameter() {defaultValue=1, group=Cubes Wrap, parameterName=Cubes_Blend_Contrast, sortPriority=-1}
    B=ScalarParameter() {defaultValue=2, group=Cubes Wrap, parameterName=Cubes_Blend_Contrast_Mech, sortPriority=-1}
    Alpha=NamedRerouteUsage()
   )
   Fill_alternate_Axis_Opacity=@n14
   "Fill_alternate_Axis_Color_A=Roughness"=@n14
   Is_Vehicle=NamedRerouteUsage()
   Is_Mech=NamedRerouteUsage()
   Enable_Fill_Z_Weapon
   Fill_Z_Weapon_opacity
   Fill_Z_Weapon_Color
  ) {materialFunction=&/MF_TopsSideProjection_Base}
  n42 = ChannelMaskParameter(None=@n41.Combined) {group=Cubes Wrap, maskChannel=Green, parameterName=Cubes_Wobble_TimeOffset_MaskChannel, sortPriority=620, defaultValue=<rgba 0 1 0 0>}
  n43 = Add(A=@n42.R, B=Time())
  n44 = Sine(None=@n43) {period=3.7}
  n45 = Cosine(None=@n43) {period=2.3}
  n46 = ChannelMaskParameter(None=@n41.Combined) {group=Cubes Wrap, parameterName=Cubes_Wobble_Multiplier_MaskChannel, sortPriority=610}
  n47 = ScalarParameter() {defaultValue=0.005, group=Cubes Wrap, parameterName=Wiggle Amount}
  n48 = Add(
   A=Multiply(
    A=MaterialFunctionCall(X=@n44, Y=@n44, Z=@n45) {materialFunction=&MakeFloat3}.Result
    B=Multiply(A=@n46.R, B=@n47)
   ) {constB=0.1}
   B=@n40
  )
  n49 = NamedRerouteUsage()
  n50 = MaterialFunctionCall(
   IsCharacter=@n6
   IsStatsB=@n7
   Texture=@n8
   Rotation_Axis=@n20
   Caculated_Bounds=@n48
   Pre_Skinned_Normals=@n49
   Blend_Contrast=@n9
   Fill_alternate_Axis_Opacity=@n10
   "Fill_alternate_Axis_Color_A=Roughness"=@n12
   Is_Vehicle=NamedRerouteUsage()
   Is_Mech=NamedRerouteUsage()
  ) {materialFunction=&MF_TopsSideProjection_Base}
  Multiply(A=@n50.RGB, B=ScalarParameter() {defaultValue=1, group=Cubes Wrap Emissive, parameterName=Cubes-Emissive01-Brightness, sortPriority=-8})
  n51 = NamedRerouteUsage()
  n52 = NamedRerouteUsage()
  n53 = TextureObjectParameter() {group=Cubes Wrap, parameterName=Cubes_BaseColor_02, sortPriority=59, texture=&T_BlacksRGB}
  n54 = NamedRerouteUsage()
  n55 = LinearInterpolate(
   A=ScalarParameter() {group=Cubes Wrap, parameterName=Cubes_Blend_Contrast, sortPriority=-1}
   B=ScalarParameter() {group=Cubes Wrap, parameterName=Cubes_Blend_Contrast_Mech, sortPriority=-1}
   Alpha=NamedRerouteUsage()
  )
  n56 = ScalarParameter() {group=Cubes Wrap, parameterName=Cubes_Fill_alternate_Axis_Opacity 02, sortPriority=51}
  n57 = VectorParameter() {group=Cubes Wrap, parameterName=Cubes_Fill_alternate_Axis_Color_Mask 02, sortPriority=67, defaultValue=<rgba 0 0 0 1>}
  n58 = AppendVector(A=@n57.RGB, B=@n57.RGB)
  n59 = NamedRerouteUsage()
  n60 = NamedRerouteUsage()
  n61 = VectorParameter() {group=Cubes Wrap UVs, parameterName=Cubes-Rotation-Weapon-02---------------------, sortPriority=150, defaultValue=<rgba 0 0 -0.25 1>}
  n62 = VectorParameter() {group=Cubes Wrap UVs, parameterName=Cubes_Rotation_Vehicle_02, sortPriority=151, defaultValue=<rgba 0 0 0 1>}
  n63 = VectorParameter() {
   desc=must activate TS_Vehicle_Side_02 UVs
   group=Cubes Wrap UVs
   parameterName=Cubes_Rotation_Vehicle_Side_02
   sortPriority=151
   defaultValue=<rgba 0 0 0 1>
  }
  n64 = Multiply(
   A=NamedRerouteUsage()
   B=ScalarParameter() {group=Cubes Wrap UVs, parameterName="Cubes_Vehicle_Side_02 UVs  0=Off 1=Active", sortPriority=149}
  )
  n65 = VectorParameter() {group=Cubes Wrap UVs, parameterName=Cubes_Rotation_Mech_02, sortPriority=153, defaultValue=<rgba 0.25 0.25 0 1>}
  n66 = LinearInterpolate(
   A=LinearInterpolate(A=@n61.RGB, B=LinearInterpolate(A=@n62.RGB, B=@n63.RGB, Alpha=@n64), Alpha=NamedRerouteUsage())
   B=@n65.RGB
   Alpha=NamedRerouteUsage()
  )
  n67 = NamedRerouteUsage()
  n68 = VectorParameter() {group=Cubes Wrap UVs, parameterName=Cubes_Offset_Weapon_02---------------------, sortPriority=160, defaultValue=<rgba 0 0 0 1>}
  n69 = VectorParameter() {group=Cubes Wrap UVs, parameterName=Cubes_Offset_Vehicle_02, sortPriority=161, defaultValue=<rgba 0 0 0 1>}
  n70 = VectorParameter() {
   desc=must activate TS_Vehicle_Side_02 UVs
   group=Cubes Wrap UVs
   parameterName=Cubes_Offset_Vehicle_Side_02
   sortPriority=161
   defaultValue=<rgba 0 0 0 1>
  }
  n71 = VectorParameter() {group=Cubes Wrap UVs, parameterName=Cubes_Offset_Mech_02, sortPriority=162, defaultValue=<rgba 0 0 0 1>}
  n72 = LinearInterpolate(
   A=LinearInterpolate(A=@n68.RGB, B=LinearInterpolate(A=@n69.RGB, B=@n70.RGB, Alpha=@n64), Alpha=NamedRerouteUsage())
   B=@n71.RGB
   Alpha=NamedRerouteUsage()
  )
  n73 = Time()
  n74 = VectorParameter() {group=Cubes Wrap UVs, parameterName=Cubes_Panning_Weapon_02---------------------, sortPriority=180, defaultValue=<rgba 0 0 0 1>}
  n75 = VectorParameter() {group=Cubes Wrap UVs, parameterName=Cubes_Panning_Vehicle_02, sortPriority=181, defaultValue=<rgba 0 0 0 1>}
  n76 = VectorParameter() {
   desc=must activate TS_Vehicle_Side_02 UVs
   group=Cubes Wrap UVs
   parameterName=Cubes_Panning_Vehicle_Side_02
   sortPriority=181
   defaultValue=<rgba 0 0 0 1>
  }
  n77 = VectorParameter() {group=Cubes Wrap UVs, parameterName=Cubes_Panning_Mech_02, sortPriority=182, defaultValue=<rgba 0 0 0 1>}
  n78 = LinearInterpolate(
   A=LinearInterpolate(A=@n74.RGB, B=LinearInterpolate(A=@n75.RGB, B=@n76.RGB, Alpha=@n64), Alpha=NamedRerouteUsage())
   B=@n77.RGB
   Alpha=NamedRerouteUsage()
  )
  n79 = VectorParameter() {group=Cubes Wrap UVs, parameterName="Cubes_Scale_Weapon_02 A=uniform", sortPriority=170, defaultValue=<rgba 1 1 1 1>}
  n80 = Multiply(A=@n79.RGB, B=@n79.RGB)
  n81 = VectorParameter() {group=Cubes Wrap UVs, parameterName="Cubes_Scale_Vehicle_02 A=uniform", sortPriority=171, defaultValue=<rgba 1 1 1 1>}
  n82 = VectorParameter() {
   desc=must activate TS_Vehicle_Side_02 UVs
   group=Cubes Wrap UVs
   parameterName="Cubes_Scale_Vehicle_Side_02 A=uniform"
   sortPriority=171
   defaultValue=<rgba 1 1 1 1>
  }
  n83 = LinearInterpolate(A=Multiply(A=@n81.RGB, B=@n81.RGB), B=Multiply(A=@n82.RGB, B=@n82.RGB), Alpha=@n64)
  n84 = NamedRerouteUsage()
  n85 = VectorParameter() {group=Cubes Wrap UVs, parameterName="Cubes_Scale_Mech_02 A=uniform", sortPriority=172, defaultValue=<rgba 1 1 1 1>}
  n86 = Multiply(
   A=Add(A=Add(A=@n67, B=@n72), B=Multiply(A=@n73, B=@n78))
   B=LinearInterpolate(A=LinearInterpolate(A=@n80, B=@n83, Alpha=@n84), B=Multiply(A=@n85.RGB, B=@n85.RGB), Alpha=NamedRerouteUsage())
  )
  NamedRerouteDeclaration(None=Multiply(
    A=MaterialFunctionCall(
     IsCharacter=@n51
     IsStatsB=@n52
     Texture=@n53
     Rotation_Axis=@n66
     Caculated_Bounds=@n86
     Pre_Skinned_Normals=@n54
     Blend_Contrast=@n55
     Fill_alternate_Axis_Opacity=@n56
     "Fill_alternate_Axis_Color_A=Roughness"=@n58
     Is_Vehicle=@n59
     Is_Mech=@n60
    ) {materialFunction=&MF_TopsSideProjection_Base}.RGB
    B=ScalarParameter() {defaultValue=1, group=Cubes Wrap, parameterName=Background Color Multiplier, sortPriority=60}
   )) {name=Base Color 2, nodeColor=<rgba 0.3125 0 1 1>}
  n87 = TextureObjectParameter() {
   group=Cubes Wrap
   parameterName=Cubes_Mask_01
   samplerType=SAMPLERTYPE_Masks
   sortPriority=20
   texture=&T_BlackMask
  }
  n88 = MaterialFunctionCall(
   IsCharacter=@n6
   IsStatsB=@n7
   Texture=@n87
   Rotation_Axis=@n20
   Caculated_Bounds=@n48
   Pre_Skinned_Normals=@n49
   Blend_Contrast=@n9
   Fill_alternate_Axis_Opacity=@n10
   "Fill_alternate_Axis_Color_A=Roughness"=@n12
   Is_Vehicle=NamedRerouteUsage()
   Is_Mech=NamedRerouteUsage()
  ) {materialFunction=&MF_TopsSideProjection_Base}
  n89 = ChannelMaskParameter(None=@n88.Combined) {group=Cubes Wrap, maskChannel=Green, parameterName=Cubes_Layer1_Mask, sortPriority=26, defaultValue=<rgba 0 1 0 0>}
  NamedRerouteDeclaration(None=@n89.R) {name=Layer Mask, nodeColor=<rgba 1 0 0.8906 1>}
  NamedRerouteDeclaration(None=FunctionInput(Preview=Constant()) {inputName=isVehicle, inputType=FunctionInput_Scalar, sortPriority=5}) {name=isVehicle, nodeColor=<rgba 0 0 0 1>}
  NamedRerouteDeclaration(None=FunctionInput(Preview=Constant()) {bUsePreviewValueAsDefault=true, inputName=isMech, inputType=FunctionInput_Scalar, sortPriority=4}) {name=isMech, nodeColor=<rgba 0 0 0 1>}
  NamedRerouteDeclaration(None=FunctionInput() {bUsePreviewValueAsDefault=true, inputName=Side Project Vehicle, inputType=FunctionInput_Scalar, sortPriority=6}) {name=Side Project, nodeColor=<rgba 0 0 0 1>}
  NamedRerouteDeclaration(None=ChannelMaskParameter(None=@n88.Combined) {group=Cubes Wrap, parameterName=Cubes_MetalicChannel_Mask_01, sortPriority=26}.R) {name=Metallic, nodeColor=<rgba 1 0.4135 0.6063 1>}
  n90 = TextureObjectParameter() {group=Cubes Wrap, parameterName=Cubes_BaseColor_01, sortPriority=11, texture=&T_BlacksRGB}
  n91 = VectorParameter() {group=Cubes Wrap, parameterName="Cubes_Fill_alternate_Axis_Color_A=Roughness 01", sortPriority=3}
  n92 = MaterialFunctionCall(
   IsCharacter=@n6
   IsStatsB=@n7
   Texture=@n90
   Rotation_Axis=@n20
   Caculated_Bounds=@n48
   Pre_Skinned_Normals=@n49
   Blend_Contrast=@n9
   Fill_alternate_Axis_Opacity=@n10
   "Fill_alternate_Axis_Color_A=Roughness"=AppendVector(A=@n91.RGB, B=@n91.RGB)
   Is_Vehicle=NamedRerouteUsage()
   Is_Mech=NamedRerouteUsage()
  ) {materialFunction=&MF_TopsSideProjection_Base}
  NamedRerouteDeclaration(None=ComponentMask(None=@n92.Combined) {b=true, g=true, r=true}) {name=Base Color, nodeColor=<rgba 1 0.4135 0.6063 1>}
  NamedRerouteDeclaration(None=LinearInterpolate(
    A=ScalarParameter() {group=Cubes Wrap, parameterName=Cubes_Roughness_Min, sortPriority=13}
    B=ScalarParameter() {defaultValue=1, group=Cubes Wrap, parameterName=Cubes_Roughness_Max, sortPriority=14}
    Alpha=ComponentMask(None=@n92.Combined) {a=true}
   )) {name=Roughness, nodeColor=<rgba 1 0.4135 0.6063 1>}
  NamedRerouteDeclaration(None=FunctionInput() {bUsePreviewValueAsDefault=true, inputName=IsStatsB, inputType=FunctionInput_Vector2}) {name=IsSTatsB, nodeColor=<rgba 0.8047 0 1 1>}
  NamedRerouteDeclaration(None=LinearInterpolate(
    A=ScalarParameter() {group=Cubes Wrap, parameterName=Cubes_Specular_Min, sortPriority=12}
    B=ScalarParameter() {defaultValue=1, group=Cubes Wrap, parameterName=Cubes_Specular_Max, sortPriority=12}
    Alpha=ChannelMaskParameter(None=@n88.Combined) {group=Cubes Wrap, parameterName=Cubes_SpecChannel_Mask_01, sortPriority=26}.R
   )) {name=Specular, nodeColor=<rgba 1 0.4135 0.6063 1>}
  NamedRerouteDeclaration(None=@n3.MaterialAttributes) {name=Material Attrib, nodeColor=<rgba 1 0.4135 0.6063 1>}
  n93 = OneMinus(None=NamedRerouteUsage())
  n94 = Multiply(A=Time(), B=ScalarParameter() {defaultValue=0.1, group=Cubes Wrap Emissive, parameterName=Cubes-Emissive01-TimeMultiplier, sortPriority=9})
  n95 = Divide(A=Add(A=Sine(None=@n94)))
  n96 = ScalarParameter() {defaultValue=0.3, group=Cubes Wrap Emissive, parameterName=Cubes-Emissive01-Min Brightness, sortPriority=11}
  n97 = ScalarParameter() {defaultValue=1, group=Cubes Wrap Emissive, parameterName=Cubes-Emissive01-Max Brightness, sortPriority=12}
  n98 = TextureObjectParameter() {
   group=Cubes Wrap
   parameterName=Cubes_Mask_02
   samplerType=SAMPLERTYPE_Masks
   sortPriority=62
   texture=&T_BlackMask
  }
  n99 = MaterialFunctionCall(
   IsCharacter=@n51
   IsStatsB=@n52
   Texture=@n98
   Rotation_Axis=@n66
   Caculated_Bounds=@n86
   Pre_Skinned_Normals=NamedRerouteUsage()
   Blend_Contrast=@n55
   Fill_alternate_Axis_Opacity=@n56
   "Fill_alternate_Axis_Color_A=Roughness"=@n58
   Is_Vehicle=NamedRerouteUsage()
   Is_Mech=NamedRerouteUsage()
  ) {materialFunction=&MF_TopsSideProjection_Base}
  n100 = ChannelMaskParameter(None=@n99.Combined) {group=Cubes Wrap Emissive, parameterName=Cubes_Emissive02_MaskChannel_02, sortPriority=105}
  n101 = VectorParameter() {group=Cubes Wrap Emissive, parameterName=Cubes_Emissive01_Color, sortPriority=8, defaultValue=<rgba 0 0 0 1>}
  n102 = Multiply(A=Multiply(A=Clamp(None=@n95, Min=@n96, Max=@n97) {minDefault=0.3}, B=@n100.R), B=@n101.RGB)
  n103 = VectorParameter() {group=Cubes Wrap Emissive, parameterName=Cubes_Emissive02_Color, sortPriority=102, defaultValue=<rgba 0 0 0 1>}
  n104 = NamedRerouteUsage()
  n105 = ScalarParameter() {group=Cubes Wrap Emissive, parameterName=Cubes_Blend_Contrast, sortPriority=-1}
  n106 = MaterialFunctionCall(In=Abs(None=ComponentMask(None=@n104) {b=true}), Contrast=@n105) {materialFunction=&CheapContrast}
  n107 = MaterialFunctionCall(In=Abs(None=ComponentMask(None=@n104) {g=true}), Contrast=@n105) {materialFunction=&CheapContrast}
  n108 = Saturate(None=Subtract(A=@n106.Result, B=@n107.Result))
  n109 = MaterialFunctionCall(In=Abs(None=ComponentMask(None=@n104) {r=true}), Contrast=@n105) {materialFunction=&CheapContrast}
  n110 = Add(
   A=Saturate(None=Add(A=@n108, B=@n109.Result))
   B=ScalarParameter() {defaultValue=1, group=Cubes Wrap Emissive, parameterName=Cubes_AEmissive_alternate_Axis_Masking_Weapon, sortPriority=3}
  )
  n111 = Saturate(None=Subtract(A=@n106.Result, B=@n109.Result))
  n112 = Add(
   A=Saturate(None=Add(A=@n111, B=@n107.Result))
   B=ScalarParameter() {defaultValue=1, group=Cubes Wrap Emissive, parameterName=Cubes_AEmissive_alternate_Axis_Masking_Vehicle, sortPriority=3}
  )
  n113 = NamedRerouteUsage()
  n114 = Saturate(None=LinearInterpolate(A=LinearInterpolate(A=@n110, B=@n112, Alpha=@n113), Alpha=NamedRerouteUsage()))
  n115 = Time()
  n116 = ScalarParameter() {group=00-Reactive-GTFB, parameterName=00-GameTimeFromBP}
  n117 = Saturate(None=Divide(
    A=Subtract(A=@n115, B=@n116)
    B=ScalarParameter() {defaultValue=1, group=Cubes Wrap Reactive, parameterName=Cubes_AnimationDuration, sortPriority=33}
   ))
  n118 = ScalarParameter() {defaultValue=0.5, group=Cubes Wrap Reactive, parameterName=Cubes_Curve_Amount, sortPriority=34}
  n119 = ScalarParameter() {group=Cubes Wrap Reactive, parameterName=Cubes_TestReactive, sortPriority=300}
  n120 = LinearInterpolate(A=OneMinus(None=Power(Base=@n117, Exp=@n118)) {desc=inverts curve from 0-1 to 1-0}, B=@n119, Alpha=@n119)
  n121 = AppendVector(
   A=LinearInterpolate(B=ScalarParameter() {defaultValue=0.3, group=Cubes Wrap Reactive, parameterName=Cubes_Time_Glow_SpeedAmount, sortPriority=603}, Alpha=@n120) {constB=0.5}
   B=Constant()
  )
  n122 = VectorParameter() {group=Cubes Wrap Emissive, parameterName=Cubes_Emissive02_OffsetWeapon, sortPriority=124, defaultValue=<rgba 0 -1 0 0>}
  n123 = ScalarParameter() {defaultValue=0.5, group=Cubes Wrap Emissive, parameterName="Cubes_Emissive02_FalloffWeapon (negative inverts)", sortPriority=123}
  n124 = VectorParameter() {group=Cubes Wrap Emissive, parameterName=Cubes_Emissive02_OffsetVehicle, sortPriority=125, defaultValue=<rgba -1 0 0 0>}
  n125 = VectorParameter() {group=Cubes Wrap Emissive, parameterName=Cubes_Emissive02_OffsetMech, sortPriority=126, defaultValue=<rgba 0 0 -1 0>}
  n126 = VectorParameter() {group=Cubes Wrap Emissive, parameterName=Cubes_Emissive02_Spread_XY, sortPriority=120, defaultValue=<rgba 1 1 1 0>}
  n127 = VectorParameter() {group=Cubes Wrap Emissive, parameterName=Cubes_Emissive02_Spread_Mech_XY, sortPriority=119, defaultValue=<rgba 1 1 1 0>}
  n128 = NamedRerouteUsage()
  n129 = VectorParameter() {group=Cubes Wrap Emissive, parameterName=Cubes_Emissive02_Animated_WeaponXY, sortPriority=130}
  n130 = VectorParameter() {group=Cubes Wrap Emissive, parameterName=Cubes_Emissive02_Animated_VehicleXY, sortPriority=131}
  n131 = NamedRerouteUsage()
  n132 = MaterialFunctionCall!(
   ForwardAxis=NamedRerouteUsage()
   Bounds=NamedRerouteUsage()
   Pre-Skinned Local Position
   isVehicle=NamedRerouteUsage()
   isMech=NamedRerouteUsage()
   Side Project Vehicle=NamedRerouteUsage()
   WeaponOffset=@n122.RGB
   WeaponFalloff=@n123
   VehicleOffset=@n124.RGB
   VehicleFalloff=ScalarParameter() {defaultValue=0.5, group=Cubes Wrap Emissive, parameterName="Cubes_Emissive02_FalloffVehicle ", sortPriority=124}
   MechOffset=@n125.RGB
   MechFalloff=ScalarParameter() {defaultValue=0.5, group=Cubes Wrap Emissive, parameterName=Cubes_Emissive02_FalloffMech, sortPriority=125}
   Invert_Gradent=StaticBoolParameter() {group=Cubes Wrap Emissive, parameterName=Cubes_Emissive02_Invert_Gradent, sortPriority=110}
   UV_Spread_XY=LinearInterpolate(A=@n126.RGB, B=@n127.RGB, Alpha=@n128)
   UV_Animated_XY=LinearInterpolate(
    A=LinearInterpolate(A=@n129.RGB, B=@n130.RGB, Alpha=@n131)
    B=VectorParameter() {group=Cubes Wrap Emissive, parameterName=Cubes_Emissive02_Animated_MechXY, sortPriority=132}.RGB
    Alpha=@n128
   )
   Warp_UVs=Multiply(
    A=ChannelMaskParameter(None=@n99.Combined) {group=Cubes Wrap Emissive, parameterName=Cubes_WarpEmissive02MaskChannel_02, sortPriority=137}.R
    B=ScalarParameter() {group=Cubes Wrap Emissive, parameterName=Cubes_WarpEmissive02_Warp_Amount, sortPriority=138}
   )
   Rotate_Mech=ScalarParameter() {defaultValue=1, group=Cubes Wrap Emissive, parameterName=Cubes_Emissive02_Texture_Rotate_Mech, sortPriority=113}
   Rotate_Vehicle_Side=ScalarParameter() {defaultValue=1, group=Cubes Wrap Emissive, parameterName=Cubes_Emissive02_Texture_Rotate_Vehicle_Side, sortPriority=113}
   Rotate_Vehicle=ScalarParameter() {defaultValue=1, group=Cubes Wrap Emissive, parameterName=Cubes_Emissive02_Texture_Rotate_Vehicle, sortPriority=113}
  ) {materialFunction=&/MF_Wrap_Gradent_Creation}
  n133 = TextureObjectParameter() {
   group=Cubes Wrap Emissive
   parameterName=Cubes_Emissive02_Gradent_Texture
   samplerType=SAMPLERTYPE_Masks
   texture=/Game/Global/Textures/Default/Masks_and_Gradients/T_VertGradient_MASK
  }
  n134 = ChannelMaskParameter(None=TextureSample(UVs=Add(A=@n121, B=@n132.Uvs), Tex=@n133) {samplerType=SAMPLERTYPE_Masks}.RGBA) {group=Cubes Wrap Emissive, parameterName=Cubes_Emissive02_TextureChannel, sortPriority=110}
  n135 = Multiply(
   A=@n103.RGB
   B=MaterialFunctionCall(
    Transition Phase=Multiply(A=@n114, B=@n134.R)
    Height Texture=@n100.R
    Contrast=ScalarParameter() {group=Cubes Wrap Emissive, parameterName=Cubes_Emissive02_HeightLerp_Contrast, sortPriority=108}
   ) {materialFunction=&HeightLerp}.Alpha
  )
  n136 = Add(
   A=@n102
   B=Multiply(A=@n135, B=ScalarParameter() {defaultValue=0.1, group=Cubes Wrap Emissive, parameterName=Cubes-Emissive01-RuneBrightness, sortPriority=10})
  )
  n137 = Add(
   A=Multiply(
    A=@n50.RGB
    B=ScalarParameter() {defaultValue=2, group=Cubes Wrap Reactive, parameterName=Cubes-Emissive01-CurveBrightness_Reactive, sortPriority=4}
   )
   B=@n135
  )
  NamedRerouteDeclaration(None=LinearInterpolate(A=Multiply(A=@n93, B=@n136), B=Add(A=@n137, B=@n102), Alpha=@n120)) {name=Emissive, nodeColor=<rgba 1 0.4135 0.6063 1>}
  n138 = NamedRerouteUsage()
  n139 = TextureObjectParameter() {
   group=Cubes Wrap
   parameterName=Cubes_Normal_01
   samplerType=SAMPLERTYPE_Normal
   sortPriority=16
   texture=&FlatNormalBlue
  }
  n140 = NamedRerouteUsage()
  n141 = NamedRerouteUsage()
  n142 = LinearInterpolate(
   A=@n138
   B=Add(
    A=@n138
    B=MaterialFunctionCall!(
     Texture_Normals=@n139
     Rotation_Axis=@n20
     Caculated_Bounds=@n48
     Pre_Skinned_Normals=@n49
     Blend_Contrast=@n9
     Fill_alternate_Axis_Opacity=@n10
     Is_Vehicle=@n140
     Is_Mech=@n141
    ) {materialFunction=&/MF_TopsSideProjection_Normals}.Add_Surface_Normal
   )
   Alpha=ScalarParameter() {defaultValue=1, group=Cubes Wrap, parameterName=Cubes_Normal_Intensty, sortPriority=1}
  )
  NamedRerouteDeclaration(None=ShadingPathSwitch(Default=@n142, Deferred=@n142, Mobile=@n138) {desc=Migrated from FeatureLevelSwitch}) {name=Normal, nodeColor=<rgba 1 0.4135 0.6063 1>}
  NamedRerouteDeclaration(None=@n3.Normal) {name=Normal Input, nodeColor=<rgba 0.5703 1 0 1>}
  n143 = VertexInterpolator(VS=PreSkinnedPosition())
  NamedRerouteDeclaration(None=MaterialFunctionCall!(
    is Vehicle=NamedRerouteUsage()
    Object Bounds Max Adjustment=FunctionInput() {bUsePreviewValueAsDefault=true, inputName=Object Bounds Max Adjustment, sortPriority=13}
    Object Bounds Min Adjustment=FunctionInput() {bUsePreviewValueAsDefault=true, inputName=Object Bounds Min Adjustment, sortPriority=12}
    Object Bounds Center Adjustment=FunctionInput() {bUsePreviewValueAsDefault=true, inputName=Object Bounds Center Adjustment, sortPriority=14}
    isMech=NamedRerouteUsage()
    Side Project Vehicle
    Local Position=FunctionInput(Preview=@n143.PS) {bUsePreviewValueAsDefault=true, inputName=Local Position, sortPriority=11}
   ) {materialFunction=&/MF_WrapBounds}.Result) {name=Bound, nodeColor=<rgba 0 0 0 1>}
  n144 = PreSkinnedNormal()
  NamedRerouteDeclaration(None=FunctionInput(Preview=VertexInterpolator(VS=@n144).PS) {bUsePreviewValueAsDefault=true, inputName=Pre-Skinned Local Normal, sortPriority=10}) {name=Preskinned Normals, nodeColor=<rgba 0 0 0 1>}
  NamedRerouteDeclaration(None=FunctionInput(Preview=Constant3Vector() {constant=<rgba 1 0 0 0>}) {bUsePreviewValueAsDefault=true, inputName=ForwardAxis}) {name=Forward Axis, nodeColor=<rgba 0 0 0 1>}
  NamedRerouteDeclaration(None=FunctionInput() {bUsePreviewValueAsDefault=true, inputName=IsCharacter, inputType=FunctionInput_Scalar}) {name=IsCharacter, nodeColor=<rgba 1 0.3047 0 1>}
 ]}
