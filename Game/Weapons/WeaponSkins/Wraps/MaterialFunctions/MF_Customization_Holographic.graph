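# Material function for weapon-skin wraps that turns input material attributes into a glowing holographic look when the Holographic Wrap static switch is on.
# MF_WrapGradient blends front/rear main and edge colors, mixed by the input normal against a fixed downward light vector.
# Screen-space scanlines and Fresnel modulate this, with a blend tint added, feeding emissive over black base color while keeping the normal.
# A world-height gradient branch is left unconnected to the output.
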MaterialFunction /Game/Weapons/WeaponSkins/Wraps/MaterialFunctions/MF_Customization_Holographic
graph {nodes [
  n1 = Constant()
  n2 = ScalarParameter() {defaultValue=0.35, group=Customization_Holographic, parameterName=Holo_ Metalness, sortPriority=-299}
  n3 = ScalarParameter() {defaultValue=1, group=Customization_Holographic, parameterName=Holo_ Specularity, sortPriority=-298}
  n4 = ScalarParameter() {defaultValue=0.2, group=Customization_Holographic, parameterName=Holo_ Roughness, sortPriority=-297}
  n5 = ScreenPosition()
  n6 = ScalarParameter() {
   defaultValue=0.2
   group=Customization_Holographic
   parameterName=Holo_ Scanlines Inner Thickness
   sliderMax=1
   sliderMin=0.01
   sortPriority=-199
  }
  n7 = Fresnel(ExponentIn=Constant() {r=3}) {exponent=3}
  n8 = OneMinus(None=Power(
    Base=MaterialFunctionCall!(ScanLine_Tiling_1, Scanline_PannerSpeed_1, ScanLine_Tiling, Scanline_PannerSpeed, UVs=@n5.ViewportUV) {materialFunction=/Game/Weapons/WeaponSkins/MaterialFunction/ScanLines}.Result
    Exp=Add(A=@n6, B=@n7)
   ))
  n9 = ScalarParameter() {defaultValue=0.5, group=Customization_Holographic, parameterName=Holo_ Scanlines Dark Value, sortPriority=-198}
  n10 = VectorParameter() {
   desc="Accent color color for the back of the gradient. Gradient settings can be found under the \"Gradient\" group."
   group=Customization_Holographic
   parameterName=Holo_ Edge Color Rear
   sortPriority=-449
   defaultValue=<rgba 0 2 1 1>
  }
  n11 = VectorParameter() {
   desc="Accent color for the front of the gradient. Gradient settings can be found under the \"Gradient\" group."
   group=Customization_Holographic
   parameterName=Holo_ Edge Color Front
   sortPriority=-499
   defaultValue=<rgba 0.0521 2 0.0521 1>
  }
  n12 = PreSkinnedPosition()
  n13 = MaterialFunctionCall!(
   isVehicle=FunctionInput() {inputName=isVehicle, inputType=FunctionInput_Scalar}
   Pre Skinned Local Position=FunctionInput(Preview=VertexInterpolator(VS=@n12).PS) {bUsePreviewValueAsDefault=true, inputName=Pre-Skinned Local Position, sortPriority=41}
   ForwardAxis=FunctionInput(Preview=Constant3Vector() {constant=<rgba 1 0 0 0>}) {inputName=ForwardAxis}
   Per Obj Bounds
  ) {desc=Gradient, materialFunction=&/MF_WrapGradient}
  n14 = Constant3Vector() {desc=Light Direction, constant=<rgba 0 0 -1 1>}
  n15 = FunctionInput() {inputName=Material, inputType=FunctionInput_MaterialAttributes, sortPriority=-200}
  n16 = Transform(None=GetMaterialAttributes(None=@n15) {attributeGetTypes [#1]}.Normal)
  n17 = Power(Base=Saturate(None=DotProduct(A=@n14, B=@n16))) {constExponent=0.6}
  n18 = VectorParameter() {
   desc="Main color for the back of the gradient. Gradient settings can be found under the \"Gradient\" group."
   group=Customization_Holographic
   parameterName=Holo_ Main Color Rear
   sortPriority=-450
   defaultValue=<rgba 0 0.5382 0.6458 1>
  }
  n19 = VectorParameter() {
   desc="Main color for the front of the gradient. Gradient settings can be found under the \"Gradient\" group."
   group=Customization_Holographic
   parameterName=Holo_ Main Color Front
   sortPriority=-500
   defaultValue=<rgba 0 0.6562 0.3281 1>
  }
  n20 = Multiply(
   A=Add(
    A=Clamp(None=@n8, Min=@n9)
    B=ScalarParameter() {group=Customization_Holographic, parameterName=Holo_ Scanlines Additional Brightness, sortPriority=-197}
   )
   B=LinearInterpolate(A=LinearInterpolate(A=@n18.RGB, B=@n19.RGB, Alpha=@n13.Result), B=LinearInterpolate(A=@n10.RGB, B=@n11.RGB, Alpha=@n13.Result), Alpha=@n17)
  )
  n21 = Saturate(None=LinearInterpolate(
    A=ScalarParameter() {defaultValue=-0.8, group=Customization_Holographic, parameterName=Holo_ Fresnel Core Intensity, sortPriority=-399}
    B=ScalarParameter() {defaultValue=1.5, group=Customization_Holographic, parameterName=Holo_ Fresnel Edge Intensity, sortPriority=-398}
    Alpha=@n7
   ))
  n22 = Add(
   A=@n21
   B=ScalarParameter() {defaultValue=0.14, group=Customization_Holographic, parameterName=Holo_ Fresnel Dark Value Clamp, sortPriority=-397}
  )
  n23 = VectorParameter() {
   desc=Additional color used across the entire wrap to unify gradient color scheme. Set to black to disable.
   group=Customization_Holographic
   parameterName=Holo_ Blend Color
   sortPriority=-420
   defaultValue=<rgba 0.8333 0 2 0>
  }
  n24 = Multiply(A=Add(A=@n7) {constB=-0.02}, B=@n23.RGB)
  n25 = Saturate(None=Add(A=Multiply(A=@n20, B=@n22), B=LinearInterpolate(A=@n24, Alpha=@n21) {constB=0}))
  n26 = GetMaterialAttributes(None=@n15) {attributeGetTypes [#2, #1]}
  FunctionOutput(None=StaticSwitchParameter(
    True=MakeMaterialAttributes!(
     BaseColor=@n1
     Metallic=@n2
     Specular=@n3
     Roughness=@n4
     Anisotropy
     EmissiveColor=@n25
     Opacity
     OpacityMask
     Normal=@n26.Normal
     Tangent
     WorldPositionOffset
     SubsurfaceColor
     ClearCoat
     ClearCoatRoughness
     AmbientOcclusion
     Refraction
     CustomizedUVs_0
     CustomizedUVs_1
     CustomizedUVs_2
     CustomizedUVs_3
     CustomizedUVs_4
     CustomizedUVs_5
     CustomizedUVs_6
     CustomizedUVs_7
     PixelDepthOffset
     ShadingModel
     Displacement
    )
    False=@n15
   ) {group=Customization_Wrap, parameterName=Holographic Wrap})
  n27 = Multiply(
   A=MaterialFunctionCall() {materialFunction=/Engine/Functions/Engine_MaterialFunctions02/WorldPositionOffset/ObjectScale}.Scale Z
   B=ScalarParameter() {defaultValue=300, group=Customization_Holographic, parameterName=Holo_ Gradient Size}
  )
  n28 = Subtract(A=Add(A=ActorPositionWS(), B=Multiply(A=@n27) {constB=0.5}) {constB=50}, B=WorldPosition().XYZ)
  n29 = Clamp(None=Divide(A=ComponentMask(None=@n28) {b=true}, B=@n27))
  n30 = StaticSwitchParameter(True=@n17, False=Power(Base=@n29) {constExponent=5}) {defaultValue=true, group=Customization_Holographic, parameterName=Holo_ Use Normal Map Texture}
  Reroute(None=QualitySwitch(Default=ShadingPathSwitch(Default=@n30, Mobile=@n29) {desc=Migrated from FeatureLevelSwitch}, Low=@n29))
 ]}
