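# Material function for the Shatter weapon wrap skin, blended over incoming material attributes by a customization mask behind a static switch.
# Triplanar projection in pre-skinned space supplies normals and packed masks: R/G tint a Fresnel two-color base, B sets roughness.
# R, G and alpha masks drive emissive through MF_WrapEmissive.
# The alpha glow follows time of day, peaking at night unless overridden, and also adds its own tint.
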MaterialFunction /Game/Weapons/WeaponSkins/Wraps/MaterialFunctions/MF_Customization_Shatter
graph {nodes [
  n1 = ScalarParameter() {
   defaultValue=0.15
   desc=Roughness Mask is B Channel of Diffuse texture.
   group=Shatter Wrap
   parameterName=Shatter_ Roughness Min
   sortPriority=7
  }
  n2 = ScalarParameter() {
   defaultValue=0.4
   desc=Roughness Mask is B Channel of Diffuse texture.
   group=Shatter Wrap
   parameterName=Shatter_ Roughness max
   sortPriority=8
  }
  n3 = FunctionInput() {inputName=Material, inputType=FunctionInput_MaterialAttributes, sortPriority=-200}
  n4 = TextureObjectParameter() {
   desc=Roughness in Alpha Channel
   group=Primary Triplaner
   parameterName=Wrap_Diffuse
   sortPriority=-20
   texture=&T_BlacksRGB
  }
  n5 = TextureObjectParameter() {
   group=Customization_Wrap
   parameterName=Wrap_Normals
   samplerType=SAMPLERTYPE_Normal
   sortPriority=-19
   texture=&FlatNormalBlue
  }
  n6 = FunctionInput() {bUsePreviewValueAsDefault=true, inputName=Wrap ScaleMultiplier, inputType=FunctionInput_Scalar, previewValue=<1 0 0 1>}
  n7 = FunctionInput(Preview=VertexInterpolator(VS=PreSkinnedPosition()).PS) {bUsePreviewValueAsDefault=true, inputName=Pre-Skinned Local Position, sortPriority=41}
  n8 = MaterialFunctionCall!(
   Material Normals=GetMaterialAttributes(None=@n3) {attributeGetTypes [#1]}.Normal
   TextureObject_MASK
   TextureObject_Diffuse=@n4
   TextureObject_Normal=@n5
   TextureScale=Multiply(A=ScalarParameter() {defaultValue=6.5, group=Primary Triplaner, parameterName=Wrap_Texture Scale, sortPriority=-18}, B=@n6)
   Normal Texture Intensity=ScalarParameter() {defaultValue=1, group=Primary Triplaner, parameterName=Wrap_Normal Texture Intensity, sortPriority=-17}
   Pre-Skinned Local Position=@n7
   Pre-Skinned Local Normal=FunctionInput() {bUsePreviewValueAsDefault=true, inputName=Pre-Skinned Local Normal, sortPriority=42}
   Rotation Z Axis
   Rotation X Axis
   Triplanner_Offset
   isMech
   Export Float 4=StaticBool() {value=true}
   ProjectionTransitionContrast=ScalarParameter() {defaultValue=1, group=Primary Triplaner, parameterName=Wrap_Triplanar Seam Hardness, sortPriority=-16}
   IsStatsB
   isCharacter
  ) {materialFunction=/Game/Packages/Fortress_SharedMaterials/Base_Material_Functions/BaseCustomizationFunctions/MF_Base_Customization_Triplanar}
  n9 = MaterialFunctionCall(Float3=@n8.Diffuse Texture) {materialFunction=&BreakOutFloat3Components}
  n10 = VectorParameter() {
   desc=Inner core of fresnel Diffuse color
   group=Shatter Wrap
   parameterName=Shatter_ Base Color 1
   sortPriority=1
   defaultValue=<rgba 0.0621 0.033 0.1094 0>
  }
  n11 = VectorParameter() {
   desc=Outer edge of fresnel Diffuse color
   group=Shatter Wrap
   parameterName=Shatter_ Base Color 2
   sortPriority=2
   defaultValue=<rgba 0.082 0.1248 0.2604 0>
  }
  n12 = LinearInterpolate(A=@n10.RGB, B=@n11.RGB, Alpha=Fresnel(Normal=Transform(None=@n8.Normal Texture)) {baseReflectFraction=0, exponent=1}) {constB=0}
  n13 = VectorParameter() {
   desc=Mask is G Channel of Diffuse texture.
   group=Shatter Wrap
   parameterName=Shatter_ Diffuse G Channel Color
   sortPriority=4
   defaultValue=<rgba 0.1093 0.0551 0.3021 0>
  }
  n14 = VectorParameter() {
   desc=Mask is R Channel of Diffuse texture.
   group=Shatter Wrap
   parameterName=Shatter_ Diffuse R Channel Color
   sortPriority=3
   defaultValue=<rgba 0.0616 0.018 0.0677 0>
  }
  n15 = VectorParameter() {
   desc="Tints emissive comprised of masks from R, G and A channels of Diffuse texture"
   group=Shatter Wrap
   parameterName=Shatter_ Emissive Color
   sortPriority=10
   defaultValue=<rgba 0.5497 0.92 5 1>
  }
  n16 = ScalarParameter() {
   defaultValue=0.5
   desc="Only on at Night. Mask is A Channel of Diffuse texture. "
   group=Shatter Wrap
   parameterName=Shatter_ A Channel Emissive Opacity
   sortPriority=14
  }
  n17 = Constant() {r=0.3}
  n18 = Constant()
  n19 = Constant() {r=0.3}
  n20 = Constant() {r=1}
  n21 = ScalarParameter() {
   desc="Use this to override the TOD and have the A Channel Emissive always be on. \"A Channel Emissive Opacity\" will still control the strength of the emissive."
   group=Shatter Wrap
   parameterName="Shatter_ A Channel Emissive Override TOD "
   sliderMax=1
   sortPriority=15
  }
  n22 = Multiply(
   A=LinearInterpolate(
    A=MaterialFunctionCall(Value_Morning=@n17, Value_Day=@n18, Value_Evening=@n19, Value_Night=@n20) {materialFunction=&Scalar_Time_of_Day}.Result
    B=Constant() {r=1}
    Alpha=@n21
   )
   B=@n8.Roughness Texture
  )
  n23 = Constant3Vector() {constant=<rgba 0 0 1 0>}
  n24 = LinearInterpolate(
   A=Multiply(A=@n22, B=@n16)
   B=Constant() {r=-0.2}
   Alpha=Fresnel(ExponentIn=Constant() {r=0.8}, Normal=Transform(None=@n23)) {baseReflectFraction=0, exponent=1}
  ) {constB=0}
  n25 = ScalarParameter() {
   defaultValue=1
   desc=Mask is R Channel of Diffuse texture.
   group=Shatter Wrap
   parameterName=Shatter_ R Channel Emissive Opacity
   sortPriority=11
  }
  n26 = Multiply(A=@n25, B=@n9.R)
  n27 = ScalarParameter() {
   defaultValue=0.3
   desc=Mask is G Channel of Diffuse texture.
   group=Shatter Wrap
   parameterName=Shatter_ G Channel Emissive Opacity
   sortPriority=12
  }
  n28 = Multiply(A=@n27, B=@n9.G)
  n29 = VertexInterpolator(VS=PreSkinnedLocalBounds().Half Extents)
  n30 = MaterialFunctionCall!(
   UseMaskGradent
   UseMaskForEmissive=StaticBool() {value=true}
   NormalizedLocalPosition=Divide(A=@n7, B=FunctionInput(Preview=@n29.PS) {bUsePreviewValueAsDefault=true, inputName=Pre-Skinned Bounds, sortPriority=41})
   ForwardDirection=FunctionInput(Preview=Constant3Vector() {constant=<rgba 1 0 0 0>}) {inputName=ForwardAxis}
   isVehicle
   HeightMapMask
   EmissiveMask=Add(A=Saturate(None=@n24), B=Add(A=@n26, B=@n28))
   HeightMapMask_2
   WrapDiffuseMask
   isMech
   IsCharacter
   IsStatsB
  ) {materialFunction=/Game/Packages/Fortress_SharedMaterials/Base_Material_Functions/BaseCustomizationFunctions/MF_WrapEmissive}
  n31 = VectorParameter() {
   desc=Only on at Night. Mask is A Channel of Diffuse texture.
   group=Shatter Wrap
   parameterName=Shatter_ A Channel Emissive Color
   sortPriority=13
   defaultValue=<rgba 0.1189 0.0125 0.2448 1>
  }
  n32 = MakeMaterialAttributes!(
   BaseColor=LinearInterpolate(A=LinearInterpolate(A=@n12, B=@n13.RGB, Alpha=@n9.G), B=@n14.RGB, Alpha=@n9.R)
   Metallic=ScalarParameter() {defaultValue=0.6, group=Shatter Wrap, parameterName=Shatter_ Metallic, sortPriority=5}
   Specular=ScalarParameter() {defaultValue=0.5, group=Shatter Wrap, parameterName=Shatter_ Specular, sortPriority=6}
   Roughness=LinearInterpolate(A=@n1, B=@n2, Alpha=@n9.B)
   Anisotropy
   EmissiveColor=Add(A=Multiply(A=@n15.RGB, B=@n30.Result), B=Multiply(A=@n31.RGB, B=@n22))
   Opacity
   OpacityMask
   Normal=@n8.Normal Texture
   Tangent
   WorldPositionOffset
   SubsurfaceColor
   ClearCoat
   ClearCoatRoughness
   AmbientOcclusion
   Refraction
   CustomizedUVs_0
   CustomizedUVs_1
   CustomizedUVs_2
   CustomizedUVs_3
   CustomizedUVs_4
   CustomizedUVs_5
   CustomizedUVs_6
   CustomizedUVs_7
   PixelDepthOffset
   ShadingModel
   Displacement
  )
  n33 = FunctionInput() {bUsePreviewValueAsDefault=true, inputName="Customization Mask (_CM B)", inputType=FunctionInput_Scalar, sortPriority=24}
  FunctionOutput(None=StaticSwitchParameter(True=BlendMaterialAttributes(A=@n3, B=@n32, Alpha=@n33), False=@n3) {group=Customization_Wrap, parameterName=Shatter Wrap})
 ]}
