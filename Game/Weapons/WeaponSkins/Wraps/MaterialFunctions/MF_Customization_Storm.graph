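# Wrap customization function that covers weapons, vehicles and backpacks in animated storm clouds with flickering lightning, behind the Storm Wrap static switch.
# Two panning, flowmap-distorted cloud layers build height and normal, driving color, specular and fresnel, with a simpler mobile path.
# Lightning jumps UVs via a timeline texture and flickers through a curve atlas.
# Clouds blend over the tinted base by mask and height; the drip gradient is unconnected.
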MaterialFunction /Game/Weapons/WeaponSkins/Wraps/MaterialFunctions/MF_Customization_Storm
graph {nodes [
  n1 = Constant() {r=0.2}
  n2 = Constant3Vector() {constant=<rgba 1 1 0 0>}
  n3 = FunctionInput() {inputName=Material, inputType=FunctionInput_MaterialAttributes, sortPriority=-200}
  n4 = TextureCoordinate() {uTiling=3, vTiling=3}
  n5 = Time()
  n6 = Constant() {r=-0.01}
  n7 = FunctionInput(Preview=Constant()) {bUsePreviewValueAsDefault=true, inputName=IsVehicle, inputType=FunctionInput_Scalar, sortPriority=-19}
  n8 = LinearInterpolate(
   A=ScalarParameter() {defaultValue=4, group=StormWrap, parameterName=CloudSpeedMaster}
   B=ScalarParameter() {defaultValue=1, group=StormWrap, parameterName=CloudSpeedMaster-Vehicle}
   Alpha=@n7
  )
  n9 = Multiply(A=Multiply(A=@n8, B=Multiply(A=@n5, B=@n6)), B=Constant2Vector() {g=1, r=0.5})
  n10 = FunctionInput(Preview=Constant() {r=0.8}) {
   bUsePreviewValueAsDefault=true
   desc=VEHICLE wrap scale
   inputName=Wrap ScaleMultiplier
   inputType=FunctionInput_Scalar
   previewValue=<1 0 0 1>
  }
  n11 = LinearInterpolate(
   A=LinearInterpolate(A=ScalarParameter() {defaultValue=0.4, group=StormWrap, parameterName=Non-VehicleScale}, B=@n10, Alpha=@n7)
   B=ScalarParameter() {defaultValue=0.4, group=StormWrap, parameterName=Backpack Scaling}
   Alpha=NamedRerouteUsage()
  )
  n12 = Multiply(A=@n8, B=Multiply(A=Constant() {r=0.015}, B=Time()))
  n13 = TextureCoordinate()
  n14 = TextureSample(UVs=Multiply(A=@n11, B=Add(A=@n12, B=@n13))) {
   samplerSource=SSM_Wrap_WorldGroupSettings
   samplerType=SAMPLERTYPE_LinearColor
   texture=/Game/VisualThreatManager/StormVisuals/LargeClouds/T_Swirl_Flowmap
  }
  n15 = Multiply(A=Constant() {r=0.4}, B=Subtract(A=ComponentMask(None=@n14.RGB) {g=true, r=true}) {constB=0.5})
  n16 = Add(A=@n15, B=Multiply(A=@n11, B=Add(A=@n4, B=@n9)))
  n17 = TextureSample(UVs=@n16) {
   samplerSource=SSM_Wrap_WorldGroupSettings
   samplerType=SAMPLERTYPE_Normal
   texture=/Game/Global/Textures/Default/Tileable_Patterns/T_StormCloudDetail01_N
  }
  n18 = Multiply(A=@n8, B=Multiply(A=Constant() {r=-0.02}, B=Time()))
  n19 = TextureCoordinate() {uTiling=4, vTiling=4}
  n20 = TextureSample(UVs=Multiply(A=Add(A=@n18, B=@n19), B=@n11)) {
   samplerSource=SSM_Wrap_WorldGroupSettings
   samplerType=SAMPLERTYPE_LinearColor
   texture=&T_Swirl_Flowmap
  }
  n21 = ComponentMask(None=@n20.RGB) {g=true, r=true}
  n22 = Multiply(A=Time(), B=Constant() {r=0.02})
  n23 = Add(A=TextureCoordinate() {uTiling=4, vTiling=4}, B=Multiply(A=Multiply(A=@n8, B=@n22), B=Constant2Vector() {g=1, r=0.5}))
  n24 = Add(A=Multiply(A=Constant() {r=0.3}, B=Subtract(A=@n21) {constB=0.5}), B=Add(A=Multiply(A=@n23, B=@n11), B=@n15))
  n25 = TextureSample(UVs=@n24) {
   samplerSource=SSM_Wrap_WorldGroupSettings
   samplerType=SAMPLERTYPE_Normal
   texture=&T_StormCloudDetail01_N
  }
  n26 = TextureSample(UVs=@n24) {
   samplerSource=SSM_Wrap_WorldGroupSettings
   samplerType=SAMPLERTYPE_LinearGrayscale
   texture=/Game/Global/Textures/Default/Tileable_Patterns/T_StormCloudDetail01_M
  }
  n27 = Constant() {r=2}
  n28 = Saturate(None=Add(A=Multiply(A=@n26.R, B=@n27) {constB=2}, B=Constant() {r=-0.5}))
  n29 = Normalize(VectorInput=Multiply(
    A=LinearInterpolate(A=@n17.RGB, B=@n25.RGB, Alpha=@n28)
    B=VectorParameter() {group=StormWrap, parameterName=CloudNormalAmount, defaultValue=<rgba 1 1 0.75 1>}.RGB
   ))
  Add(
   A=Multiply(
    A=GetMaterialAttributes(None=@n3) {attributeGetTypes [#1, #2, #3]}.Normal
    B=Multiply(A=@n1, B=@n2)
   )
   B=@n29
  )
  Constant3Vector() {constant=<rgba 0.1059 0.09 0.3 1>}
  Constant3Vector() {constant=<rgba 9.9872 5 10 1>}
  Constant3Vector() {constant=<rgba 0.0318 0.0163 0.1 1>}
  Constant() {r=2.633}
  Constant() {r=0.15}
  Constant() {r=0.25}
  Constant() {r=1}
  Constant()
  Constant() {r=0.8}
  Constant() {r=0.33}
  Constant() {r=5.0061}
  Constant() {r=0.5}
  FunctionInput(Preview=VertexInterpolator(VS=PreSkinnedNormal()).PS) {bUsePreviewValueAsDefault=true, inputName=PreSkinnedLocalNormal}
  n30 = MaterialFunctionCall!(Material=@n3) {materialFunction=/Engine/Functions/MaterialLayerFunctions/MatLayerBlend_BreakBaseColor}
  n31 = ScalarParameter() {defaultValue=1, group=StormWrap, parameterName=BaseColorInput-Desaturate}
  n32 = FunctionInput() {bUsePreviewValueAsDefault=true, inputName="Customization Mask (_CM B)", inputType=FunctionInput_Scalar, sortPriority=24}
  n33 = MaterialFunctionCall!(
   Material=@n3
   NewBaseColor=Multiply(
    A=Desaturation(None=@n30.BaseColor, Fraction=@n31)
    B=VectorParameter() {group=StormWrap, parameterName=BaseColorInput-Mod, defaultValue=<rgba 0.1591 0.0816 0.5 1>}.RGB
   )
   Mask=@n32
  ) {materialFunction=/Engine/Functions/MaterialLayerFunctions/MatLayerBlend_OverrideBaseColor}
  n34 = ScalarParameter() {defaultValue=-1.5, group=StormWrap, parameterName=StormBlendLow}
  n35 = ScalarParameter() {defaultValue=2, group=StormWrap, parameterName=StormBlendHigh}
  n36 = TextureSample(UVs=@n16) {
   samplerSource=SSM_Wrap_WorldGroupSettings
   samplerType=SAMPLERTYPE_LinearGrayscale
   texture=/Game/Global/Textures/Default/Flowmaps/T_StormSwirlMask
  }
  n37 = ShadingPathSwitch(Default=LinearInterpolate(A=@n36.R, B=@n26.R, Alpha=@n28), Mobile=@n26.R) {desc=Migrated from FeatureLevelSwitch}
  n38 = Multiply(A=@n32, B=Saturate(None=LinearInterpolate(A=@n34, B=@n35, Alpha=@n37)))
  n39 = Power(Base=@n37, Exp=ScalarParameter() {defaultValue=4, group=StormWrap, parameterName=CloudHeightPowerForDiffuse})
  n40 = ShadingPathSwitch(Default=@n29, Mobile=Constant3Vector() {constant=<rgba 0 0 1 0>}) {desc=Migrated from FeatureLevelSwitch}
  n41 = Saturate(None=DotProduct(A=Transform(None=@n40), B=CameraVectorWS()))
  n42 = Multiply(
   A=Power(Base=OneMinus(None=@n41), Exp=ScalarParameter() {defaultValue=2, group=StormWrap, parameterName=FresnelPower})
   B=VectorParameter() {group=StormWrap, parameterName=CloudFresnelColor, defaultValue=<rgba 0.2 0.5289 1 1>}.RGB
  )
  n43 = ScalarParameter() {defaultValue=0.05, group=StormWrap, parameterName=FresnelEmissiveAmount}
  n44 = VectorParameter() {group=StormWrap, parameterName=StormLightningColor, defaultValue=<rgba 9.9872 5 10 1>}
  n45 = ScalarParameter() {defaultValue=5.0061, group=StormWrap, parameterName=LNG-UVOffset}
  n46 = ScalarParameter() {defaultValue=0.33, group=StormWrap, parameterName=LNG-TimeShiftOffset}
  n47 = Multiply(A=Time(), B=ScalarParameter() {defaultValue=0.8, group=StormWrap, parameterName=LNG-TimeShift})
  n48 = Ceil(None=@n47)
  n49 = TextureSample(UVs=AppendVector(A=Multiply(A=@n46, B=@n48), B=Constant())) {samplerType=SAMPLERTYPE_Masks, texture=/Game/Gadgets/Assets/VinderTech_GliderChute/Textures/T_LightingTimeline_M}
  n50 = Add(
   A=AppendVector(A=Multiply(A=@n45, B=@n49.G), B=Multiply(A=@n45, B=@n49.B))
   B=Multiply(A=TextureCoordinate(), B=ScalarParameter() {defaultValue=3.5, group=StormWrap, parameterName=LNG-MaskTile})
  )
  n51 = Frac(None=@n47)
  n52 = Multiply(A=@n51, B=OneMinus(None=@n51))
  n53 = ScalarParameter() {defaultValue=0.5, group=StormWrap, parameterName=LNG-FadeShiftSpeed}
  n54 = Multiply(
   A=TextureSample(UVs=Multiply(A=@n11, B=@n50)) {samplerType=SAMPLERTYPE_LinearColor, texture=/Game/Athena/SafeZone/Textures/T_Clouds_Lightning}.G
   B=Clamp(None=Multiply(A=@n52, B=@n53))
  )
  n55 = Add(A=Time(), B=ScalarParameter() {defaultValue=0.25, group=StormWrap, parameterName=LNG-RandomOffset})
  n56 = ScalarParameter() {defaultValue=0.2, group=StormWrap, parameterName=LNG-RandomSpeed}
  n57 = CurveAtlasRowParameter(CurveTime=Frac(None=Multiply(A=@n55, B=@n56))) {
   atlas=&AthenaCurveAtlas01
   curve=/Game/Athena/SafeZone/Textures/StormLighting
   defaultValue=18
   group=Customization
   parameterName=Curve1_1
  }
  n58 = Multiply(A=@n54, B=Multiply(A=@n57.R, B=@n57.R))
  n59 = Multiply(A=@n58, B=OneMinus(None=@n37))
  n60 = Multiply(A=@n42, B=ScalarParameter() {defaultValue=0.3, group=StormWrap, parameterName=FresnelDiffuseAmount})
  n61 = Multiply(A=OneMinus(None=@n58), B=Constant() {r=15.75}) {constB=2}
  n62 = Constant() {r=-15}
  n63 = Multiply(A=Saturate(None=Add(A=@n61, B=@n62)), B=@n39)
  n64 = VectorParameter() {group=StormWrap, parameterName=CloudDarkColor, defaultValue=<rgba 0.0318 0.0163 0.1 1>}
  n65 = VectorParameter() {group=StormWrap, parameterName=CloudLightColor, defaultValue=<rgba 0.1059 0.09 0.3 1>}
  n66 = SetMaterialAttributes!(
   MaterialAttributes
   BaseColor=Add(A=Multiply(A=@n60, B=@n63), B=LinearInterpolate(A=@n64.RGB, B=@n65.RGB, Alpha=@n63))
   Normal=@n40
   Specular=LinearInterpolate(
    A=ScalarParameter() {defaultValue=0.3, group=StormWrap, parameterName=StormSpecularLow}
    B=ScalarParameter() {defaultValue=0.6, group=StormWrap, parameterName=StormSpecular}
    Alpha=@n39
   )
   Roughness=ScalarParameter() {defaultValue=0.7, group=StormWrap, parameterName=StormRoughness}
   EmissiveColor=Add(A=Multiply(A=@n44.RGB, B=@n59), B=Multiply(A=@n42, B=@n43))
  ) {attributeSetTypes [
    #4
    #1
    #3
    #2
    #5
   ]}
  FunctionOutput(None=StaticSwitchParameter(True=BlendMaterialAttributes(A=@n33.Blended Material, B=@n66, Alpha=@n38), False=@n3) {group=Customization_Wrap, parameterName=Storm Wrap})
  n67 = ScalarParameter() {defaultValue=-0.3, group=StormWrap, parameterName=DripweaponOffset}
  n68 = ScalarParameter() {defaultValue=0.15, group=StormWrap, parameterName=DripVehicleOffset}
  n69 = Divide(
   A=ObjectRadius()
   B=MaterialFunctionCall() {materialFunction=/Engine/Functions/Engine_MaterialFunctions02/WorldPositionOffset/ObjectScale}.Scale X
  )
  n70 = ScalarParameter() {defaultValue=0.35, group=StormWrap, parameterName=DripObjectRadiusScale}
  n71 = PreSkinnedPosition()
  n72 = Divide(
   A=FunctionInput(Preview=VertexInterpolator(VS=@n71).PS) {bUsePreviewValueAsDefault=true, inputName=PreSkinLocalPosition}
   B=LinearInterpolate(A=ScalarParameter() {defaultValue=30, group=StormWrap, parameterName=DripWeaponDivisor}, B=Multiply(A=@n69, B=@n70))
  )
  n73 = Clamp(None=Add(A=LinearInterpolate(A=@n67, B=@n68, Alpha=@n7), B=ConstantBiasScale(None=@n72)))
  n74 = FunctionInput(Preview=Constant3Vector() {constant=<rgba 0 1 0 0>}) {inputName=ForwardAxis}
  n75 = MaterialFunctionCall(Float3=Power(Base=Multiply(A=@n73, B=@n74), Exp=ScalarParameter() {defaultValue=0.5, group=StormWrap, parameterName=DripGradient Power})) {materialFunction=&BreakOutFloat3Components}
  n76 = Saturate(None=Max(A=Max(A=@n75.R, B=@n75.R), B=@n75.B))
  n77 = OneMinus(None=@n76)
  n78 = ScalarParameter() {group=StormWrap, parameterName=Dripinvert Gradient}
  MaterialFunctionCall(Transition Phase=Saturate(None=LinearInterpolate(A=@n76, B=@n77, Alpha=@n78)), Height Texture=@n37) {materialFunction=&HeightLerp}
  Multiply(A=TextureCoordinate(), B=ScalarParameter() {defaultValue=4, parameterName=TopFlow2Tile})
  Multiply(A=TextureCoordinate(), B=ScalarParameter() {defaultValue=1, parameterName=TopFlowTile})
  Multiply(A=@n54, B=@n57.R)
  NamedRerouteDeclaration(None=FunctionInput() {bUsePreviewValueAsDefault=true, inputName=IsCharacter, inputType=FunctionInput_Scalar, sortPriority=-19}) {name=IsCharacter, nodeColor=<rgba 0 0.9253 1 1>}
  n79 = FunctionInput() {bUsePreviewValueAsDefault=true, inputName=IsStatsB, inputType=FunctionInput_Vector2, sortPriority=-19}
  NamedRerouteDeclaration(None=ComponentMask(None=@n79) {r=true}) {name=IsBackpack, nodeColor=<rgba 0 0.9253 1 1>}
  NamedRerouteDeclaration(None=ComponentMask(None=@n79) {g=true}) {name=IsPickaxe, nodeColor=<rgba 0 0.9253 1 1>}
  Reroute(None=Multiply(A=TextureCoordinate(), B=Constant() {r=4}))
  Reroute(None=Multiply(A=TextureCoordinate(), B=Constant() {r=3}))
 ]}
