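# Material function for The Cube cosmetic wrap, blending a dark crystal surface with pulsing glowing core and glyphs over the input material through a customization mask and static switch.
# Bump-offset crystal texture drives color, roughness, normal flattening and glow, while crevasse-biased fresnel defines unlit edges.
# A screen-aligned sphere mask shapes the core, which gates triplanar glyphs projected in pre-skinned space.
# Scale and tiling vary for weapon, vehicle, pickaxe, character and backpack.
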MaterialFunction /Game/Weapons/WeaponSkins/Wraps/MaterialFunctions/MF_Customization_TheCube
graph {nodes [
  Constant3Vector() {desc=sphere center location, constant=<rgba 0.5 0.5 0.5 1>}
  Constant() {desc="Sphere Mask Radius weapon ", r=0.25}
  Constant() {desc=Sphere Mask Radius vehicle, r=2.3}
  Constant() {desc=sphere mask hardness, r=-1}
  n1 = VectorParameter() {group=Cube Wrap, parameterName=Cube_ Core Diffuse Color, sortPriority=1, defaultValue=<rgba 0.0081 0.0029 0.04 1>}
  n2 = VectorParameter() {group=Cube Wrap, parameterName=Cube_ Edge Diffuse Color, sortPriority=2, defaultValue=<rgba 0 0 0 1>}
  n3 = Constant() {desc=grow curvature line edges, r=2}
  n4 = Subtract(A=Constant() {desc=shrink curvature line edges, r=0.6}, B=@n3)
  n5 = FunctionInput(Preview=TextureSample() {samplerType=SAMPLERTYPE_Masks, texture=/Game/Athena/DrivableVehicles/Golf_Cart/Textures/T_Golf_Cart_Base_CM}.G) {bUsePreviewValueAsDefault=true, inputName="Customization Crevasse (_CM G)", inputType=FunctionInput_Scalar, sortPriority=21}
  n6 = Multiply(A=Saturate(None=LinearInterpolate(A=@n4, B=@n3, Alpha=@n5)), B=Constant() {r=2})
  n7 = LinearInterpolate(
   A=Constant() {desc=edge mask fresnel min, r=-0.3}
   B=Constant() {desc=edge mask fresnel max, r=0.3}
   Alpha=Fresnel(ExponentIn=Constant() {r=2}, BaseReflectFractionIn=Saturate(None=@n6)) {exponent=1}
  )
  n8 = Saturate(None=@n7)
  n9 = MaterialFunctionCall(Alpha=OneMinus(None=Fresnel() {exponent=1}), Max=Constant() {r=2}) {materialFunction=&SmoothStep}
  n10 = FunctionInput() {inputName=isVehicle, inputType=FunctionInput_Scalar}
  n11 = LinearInterpolate(
   A=Constant() {desc=Modify tiling texture scale for weapons., r=1}
   B=Constant() {desc=Cube_ Crystal Vehicle Tile Modifier, r=1.5}
   Alpha=@n10
  )
  n12 = TextureCoordinate()
  n13 = TextureSample(UVs=BumpOffset(
    Coordinate=Multiply(A=@n11, B=@n12)
    Height=Constant() {desc=Cube_ Crystal Bump Offset Height, r=0.6}
    HeightRatioInput=Constant() {desc=Cube_ Crystal Bump Offset Height Ratio, r=-0.6}
   )) {samplerType=SAMPLERTYPE_Masks, texture=/Game/Weapons/WeaponSkins/Wraps/Textures/TheCube/T_Cube_Wrap_Masks}
  n14 = Multiply(A=@n13.R, B=@n9.Result) {desc=combine crystal texture with gentle fresnel}
  n15 = ScalarParameter() {defaultValue=0.35, group=Cube Wrap, parameterName=Cube_ Specular Outer Fresnel Value, sortPriority=3}
  n16 = ScalarParameter() {defaultValue=0.15, group=Cube Wrap, parameterName=Cube_ Specular Inner Fresnel Value, sortPriority=4}
  n17 = ScalarParameter() {defaultValue=0.42, group=Cube Wrap, parameterName=Cube_ Roughness Crystal Min, sortPriority=6}
  n18 = ScalarParameter() {defaultValue=0.55, group=Cube Wrap, parameterName=Cube_ Roughness Crystal Max, sortPriority=7}
  n19 = LinearInterpolate(A=Constant() {desc=fresnel min, r=0.2}, B=Constant() {desc=fresnel max, r=1}, Alpha=@n9.Result)
  n20 = NamedRerouteUsage()
  n21 = TransformPosition(None=ComponentMask(None=@n20) {b=true, g=true, r=true}) {transformType=TRANSFORMPOSSOURCE_World}
  n22 = ComponentMask(None=@n20) {b=true, g=true, r=true}
  n23 = NamedRerouteUsage()
  n24 = Multiply(
   A=Divide(B=MaterialFunctionCall() {materialFunction=&ObjectScale}.Scale Z)
   B=MaterialFunctionCall!(Center Point World Position=LinearInterpolate(A=@n21, B=@n22, Alpha=@n23)) {materialFunction=/Engine/Functions/Engine_MaterialFunctions02/UVs/WorldSpaceAlignedScreenCoordinates}.Offset and Scaled UVS
  )
  n25 = PreSkinnedLocalBounds()
  n26 = MaterialFunctionCall(Float3=Subtract(A=@n25.Max, B=@n25.Max)) {materialFunction=&BreakOutFloat3Components}
  n27 = Max(A=@n26.R, B=@n26.R)
  n28 = Divide(A=Constant() {r=-2500}, B=Multiply(A=ComponentMask(None=@n20) {a=true}, B=Max(A=@n27, B=@n26.B)))
  n29 = ScalarParameter() {defaultValue=1, group=Pliant, parameterName=SphereMaskGlow_Weapon}
  n30 = ScalarParameter() {defaultValue=0.25, group=Pliant, parameterName=SphereMaskGlow_Vehicle}
  n31 = LinearInterpolate(
   A=LinearInterpolate(
    A=LinearInterpolate(A=@n29, B=@n30, Alpha=@n10)
    B=ScalarParameter() {group=Pliant, parameterName=SphereMaskGlow_PickAxe}
    Alpha=NamedRerouteUsage()
   )
   B=LinearInterpolate(
    A=ScalarParameter() {group=Pliant, parameterName=SphereMaskGlow_Character}
    B=ScalarParameter() {group=Pliant, parameterName=SphereMaskGlow_Backpack}
    Alpha=NamedRerouteUsage()
   )
   Alpha=NamedRerouteUsage()
  )
  n32 = SphereMask(
   A=Add(A=Multiply(A=@n24, B=@n28)) {constB=0.5}
   B=Constant2Vector() {g=0.5, r=0.5}
   Radius=@n31
   Hardness=ScalarParameter() {defaultValue=-1, group=Pliant, parameterName=SphereMaskGlow_Hardness}
  ) {attenuationRadius=1, hardnessPercent=0.5}
  n33 = LinearInterpolate(A=Constant() {desc=inner core sphere mask min, r=-0.5}, B=Constant() {desc=inner core sphere mask max, r=4}, Alpha=@n32)
  n34 = Power(Base=Saturate(None=@n33))
  n35 = VectorParameter() {
   desc=Cube_ Emissive Color Inner Core
   group=Cube Wrap
   parameterName=Cube_ Emissive Color Inner Core
   sortPriority=10
   defaultValue=<rgba 0.6852 0 0.75 1>
  }
  n36 = Multiply(A=@n34, B=@n35.RGB)
  n37 = MaterialFunctionCall(Value=Multiply(A=Time(), B=ScalarParameter() {defaultValue=0.2, group=Cube Wrap, parameterName=Cube_ Emissive Pulse Speed, sortPriority=12})) {materialFunction=&LinearSine}
  n38 = LinearInterpolate(A=Constant() {desc=Cube_ Emissive Pulse Max, r=0.6}, B=Constant() {desc=Cube_ Emissive Pulse Min, r=0.15}, Alpha=@n37."Rounded Linear Sine ")
  n39 = VectorParameter() {
   desc=Cube_ Emissive Color Outer Core
   group=Cube Wrap
   parameterName=Cube_ Emissive Color Outer Core
   sortPriority=11
   defaultValue=<rgba 0.1607 0 0.3073 1>
  }
  n40 = VectorParameter() {group=Cube Wrap, parameterName=Cube_ Crystal Glow Color1, sortPriority=13, defaultValue=<rgba 0 0.002 0.0365 1>}
  n41 = VectorParameter() {group=Cube Wrap, parameterName=Cube_ Crystal Glow Color2, sortPriority=14, defaultValue=<rgba 0.0298 0 0.2 1>}
  n42 = VectorParameter() {group=Cube Wrap, parameterName=Cube_ Crystal Glow Color 3, sortPriority=15, defaultValue=<rgba 0.1844 0 0.4375 1>}
  n43 = VectorParameter() {group=Cube Wrap, parameterName=Cube_ Crystal Glow Color 4, sortPriority=16, defaultValue=<rgba 0.7166 0 1 1>}
  n44 = Constant() {r=1.5}
  n45 = Power(
   Base=Multiply(A=@n19, B=@n13.R) {desc=crystal texture}
   Exp=LinearInterpolate(A=Constant() {desc=texture Pulse Max, r=2.75}, B=Constant() {desc=texture pulse min, r=1.5}, Alpha=@n37."Rounded Linear Sine ")
  )
  n46 = Add(
   A=@n39.RGB
   B=MaterialFunctionCall(1=@n40.RGB, 2=@n41.RGB, 3=@n42.RGB, 4=@n43.RGB, A=Multiply(A=@n44, B=@n45)) {materialFunction=&Lerp_Multiple_Float3}.Lerp 4 Inputs
  )
  n47 = Constant() {r=1.5}
  n48 = Multiply(A=Multiply(A=@n19, B=@n32) {desc=sphere mask}, B=Add(A=Multiply(A=@n36, B=@n38), B=Multiply(A=@n46, B=@n47)))
  n49 = VectorParameter() {group=Cube Wrap, parameterName=Cube_ Glyph Glow Color, sortPriority=9, defaultValue=<rgba 0.3203 0 0.6406 1>}
  n50 = PreSkinnedPosition()
  n51 = Constant() {desc=Cube_ Glyph Texture Tiling. Amount that Glyph textures tile., r=28}
  n52 = Multiply(A=@n51, B=Constant() {desc=Cube_ Glyph Vehicle Tile Modifier, r=2.3})
  n53 = LinearInterpolate(
   A=LinearInterpolate(
    A=LinearInterpolate(A=@n51, B=@n52, Alpha=@n10)
    B=ScalarParameter() {defaultValue=20, group=Pliant, parameterName=TriplanerTilingGlyphs_PickAxe}
    Alpha=NamedRerouteUsage()
   )
   B=LinearInterpolate(
    A=ScalarParameter() {defaultValue=20, group=Pliant, parameterName=TriplanerTilingGlyphs_Character}
    B=ScalarParameter() {defaultValue=20, group=Pliant, parameterName=TriplanerTilingGlyphs_Backpack}
    Alpha=NamedRerouteUsage()
   )
   Alpha=NamedRerouteUsage()
  )
  n54 = Divide(
   A=FunctionInput(Preview=VertexInterpolator(VS=@n50).PS) {bUsePreviewValueAsDefault=true, inputName=Pre-Skinned Local Position, sortPriority=-16}
   B=Multiply(A=Abs(None=@n53)) {constB=-1}
  )
  n55 = TextureObject() {samplerType=SAMPLERTYPE_Masks, texture=&T_Cube_Wrap_Masks}
  n56 = TextureSample(UVs=ComponentMask(None=@n54) {b=true, r=true}, Tex=@n55) {samplerSource=SSM_Wrap_WorldGroupSettings}
  n57 = TextureSample(
   UVs=MaterialFunctionCall(
    UVs=ComponentMask(None=@n54) {b=true, g=true}
    "Rotation Angle (0-1)"=FunctionInput() {bUsePreviewValueAsDefault=true, inputName=RotXAxis, inputType=FunctionInput_Scalar}
   ) {materialFunction=&CustomRotator}.Rotated Values
   Tex=@n55
  ) {samplerSource=SSM_Wrap_WorldGroupSettings}
  n58 = FunctionInput(Preview=VertexInterpolator(VS=PreSkinnedNormal()).PS) {bUsePreviewValueAsDefault=true, inputName=Pre-Skinned Local Normal, sortPriority=42}
  n59 = ComponentMask(None=@n58) {r=true}
  n60 = Constant() {r=1}
  n61 = ComponentMask(None=MaterialFunctionCall(In=Abs(None=@n59), Contrast=@n60) {materialFunction=&CheapContrast}.Result) {r=true}
  n62 = MaterialFunctionCall(
   UVs=ComponentMask(None=@n54) {g=true, r=true}
   "Rotation Angle (0-1)"=FunctionInput() {bUsePreviewValueAsDefault=true, inputName=RotZAxis, inputType=FunctionInput_Scalar}
  ) {materialFunction=&CustomRotator}
  n63 = MaterialFunctionCall(In=Abs(None=ComponentMask(None=@n58) {b=true}), Contrast=@n60) {materialFunction=&CheapContrast}
  n64 = MaterialFunctionCall(Float3=LinearInterpolate(
    A=LinearInterpolate(A=@n56.RGB, B=@n57.RGB, Alpha=@n61) {desc=Float3}
    B=TextureSample(UVs=@n62.Rotated Values, Tex=@n55) {samplerSource=SSM_Wrap_WorldGroupSettings}.RGB
    Alpha=ComponentMask(None=@n63.Result) {r=true}
   )) {materialFunction=&BreakOutFloat3Components}
  n65 = LinearInterpolate(
   A=Constant() {desc=Cube_ Glpyh Pulse Flare Width Max, r=5}
   B=Constant() {desc=Cube_ Glpyh Pulse Flare Width Min, r=35}
   Alpha=@n37."Rounded Linear Sine "
  )
  n66 = Add(A=Multiply(A=Power(Base=@n64.G, Exp=@n65), B=Constant() {r=4}), B=Constant() {r=0.05})
  n67 = Multiply(A=@n64.G, B=Constant() {desc=Cube_ Glyph Soft Glow strength, r=0.1})
  n68 = LinearInterpolate(A=Constant() {desc=fresnel min}, B=Constant() {desc=fresnel max, r=1.2}, Alpha=@n9.Result)
  n69 = Multiply(A=@n49.RGB, B=LinearInterpolate(B=Add(A=@n66, B=@n67), Alpha=Multiply(A=@n68, B=@n34)))
  n70 = Constant() {r=1.5}
  n71 = Constant() {r=-2}
  n72 = FunctionInput() {inputName=Material, inputType=FunctionInput_MaterialAttributes, sortPriority=-200}
  n73 = MakeMaterialAttributes!(
   BaseColor=LinearInterpolate(A=@n1.RGB, B=@n2.RGB, Alpha=Add(A=@n8, B=@n14))
   Metallic=Constant() {r=0.2}
   Specular=LinearInterpolate(
    A=LinearInterpolate(A=@n15, B=@n16, Alpha=@n9.Result)
    B=ScalarParameter() {defaultValue=0.1, group=Cube Wrap, parameterName=Cube_ Specular Edge Value, sortPriority=5}
    Alpha=@n7
   )
   Roughness=LinearInterpolate(
    A=LinearInterpolate(A=@n17, B=@n18, Alpha=@n13.R)
    B=ScalarParameter() {defaultValue=0.8, group=Cube Wrap, parameterName=Cube_ Roughness Edge Value, sortPriority=8}
    Alpha=@n7
   )
   Anisotropy
   EmissiveColor=LinearInterpolate(A=Add(A=@n48, B=@n69), B=Constant(), Alpha=@n7)
   Opacity
   OpacityMask
   Normal=MaterialFunctionCall(
    Normal=GetMaterialAttributes(None=@n72) {attributeGetTypes [#1]}.Normal
    Flatness=LinearInterpolate(A=@n70, B=@n71, Alpha=@n13.R)
   ) {materialFunction=&FlattenNormal}.Result
   Tangent
   WorldPositionOffset
   SubsurfaceColor
   ClearCoat
   ClearCoatRoughness
   AmbientOcclusion
   Refraction
   CustomizedUVs_0
   CustomizedUVs_1
   CustomizedUVs_2
   CustomizedUVs_3
   CustomizedUVs_4
   CustomizedUVs_5
   CustomizedUVs_6
   CustomizedUVs_7
   PixelDepthOffset
   ShadingModel
   Displacement
  )
  n74 = FunctionInput() {bUsePreviewValueAsDefault=true, inputName="Customization Mask (_CM B)", inputType=FunctionInput_Scalar, sortPriority=24}
  FunctionOutput(None=StaticSwitchParameter(True=BlendMaterialAttributes(A=@n72, B=@n73, Alpha=@n74), False=@n72) {group=Customization_Wrap, parameterName=The C U B E})
  Multiply(A=@n32)
  n75 = PreSkinnedPosition()
  n76 = PreSkinnedLocalBounds()
  n77 = MaterialFunctionCall(Color=Subtract(A=@n76.Max, B=@n76.Max)) {materialFunction=&SplitComponents}
  n78 = Divide(A=Subtract(A=VertexInterpolator(VS=@n75).PS, B=@n76.Min), B=Max(A=Max(A=@n77.R, B=@n77.R), B=@n77.B))
  n79 = Constant()
  Multiply(A=MaterialFunctionCall!(
    Normal_Map
    Offset Rt
    Pivot
    Caculated_Bounds=FunctionInput(Preview=@n78) {bUsePreviewValueAsDefault=true, inputName=Calcated_Bounds, sortPriority=11}
    Pre_Skinned_Normals=NamedRerouteUsage()
    is_State=FunctionInput(Preview=@n79) {bUsePreviewValueAsDefault=true, inputName=IsStats, inputType=FunctionInput_Vector4}
    UV_Rotation
    Uvs_Offset
    Uvs_TileRate
    Uvs_Offset_Pivot
    For_Safe_Zones
    Rotation_Comp
    Clamp Y
    Clamp X
    Clamp Padding_Comp
    No_Rotation=StaticBool() {value=true}
   ) {materialFunction=&/MF_Projection_Side_Normal_V2}.UVs)
  NamedRerouteDeclaration(None=FunctionInput() {bUsePreviewValueAsDefault=true, inputName=IsCharacter, inputType=FunctionInput_Scalar}) {name=IsCharacter, nodeColor=<rgba 0 0.9253 1 1>}
  n80 = FunctionInput() {bUsePreviewValueAsDefault=true, inputName=IsStatsB, inputType=FunctionInput_Vector2}
  NamedRerouteDeclaration(None=ComponentMask(None=@n80) {r=true}) {name=IsBackpack, nodeColor=<rgba 0 0.9253 1 1>}
  n81 = FunctionInput(Preview=Constant4Vector() {constant=<rgba 0 0 5 -1>}.RGBA) {bUsePreviewValueAsDefault=true, inputName=Object-SafeZone-Virtual-Pivot, inputType=FunctionInput_Vector4, sortPriority=105}
  NamedRerouteDeclaration(None=@n81) {name=SafeZone VP, nodeColor=<rgba 0.5 1 0 1>}
  n82 = FunctionInput(Preview=Constant()) {bUsePreviewValueAsDefault=true, inputName=Object-SafeZone-VP-Character-PreView-Switch, inputType=FunctionInput_Scalar, sortPriority=105}
  NamedRerouteDeclaration(None=@n82) {name=SafeZone Character-Switch, nodeColor=<rgba 1 0.9375 0 1>}
  NamedRerouteDeclaration(None=ComponentMask(None=@n80) {g=true}) {name=IsPickaxe, nodeColor=<rgba 0 0.9253 1 1>}
  NamedRerouteDeclaration(None=@n33) {name=To Pliant Overrides, nodeColor=<rgba 1 0.0584 0 1>}
  n83 = NamedRerouteUsage()
  n84 = ScalarParameter() {defaultValue=3, group=Pliant, parameterName=SecondaryInnerCore_PickAxe}
  n85 = NamedRerouteUsage()
  n86 = ScalarParameter() {defaultValue=3, group=Pliant, parameterName=SecondaryInnerCore_Character}
  n87 = ScalarParameter() {defaultValue=3, group=Pliant, parameterName=SecondaryInnerCore_Backpack}
  n88 = NamedRerouteUsage()
  NamedRerouteDeclaration(None=LinearInterpolate(A=LinearInterpolate(A=@n83, B=@n84, Alpha=@n85), B=LinearInterpolate(A=@n86, B=@n87, Alpha=@n88), Alpha=NamedRerouteUsage())) {name=FromPliantOverrides, nodeColor=<rgba 1 0.0584 0 1>}
  NamedRerouteDeclaration(None=@n58) {name=Preskin_Normals, nodeColor=<rgba 1 0.6328 0 1>}
  NamedRerouteUsage()
  n89 = VertexInterpolator(VS=PreSkinnedLocalBounds().Half Extents)
  Reroute(None=Divide(A=FunctionInput(Preview=@n89.PS) {bUsePreviewValueAsDefault=true, inputName=Pre-Skinned Bounds, sortPriority=43}, B=@n31))
  Subtract(A=WorldPosition().XYZ, B=ObjectPositionWS())
 ]}
