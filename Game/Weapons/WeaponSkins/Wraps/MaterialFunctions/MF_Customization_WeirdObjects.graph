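# Weapon-wrap material function that layers animated, vein-like "Weird Objects" detail over an input material, with vehicle and character tuning.
# Two panning noise samples distort pre-skinned positions feeding triplanar mask and normal projection.
# That mask, a forward-axis gradient and the noise drive a HeightLerp blending two-tone Fresnel-tinted color, metallic, specular, roughness and normal.
# A static switch passes the input through when the wrap is off.
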MaterialFunction /Game/Weapons/WeaponSkins/Wraps/MaterialFunctions/MF_Customization_WeirdObjects
graph {nodes [
  Constant3Vector() {constant=<rgba 0.06 0.1338 0.2 1>}
  n1 = ScalarParameter() {defaultValue=4, group=WeirdObjects, parameterName=Distort-Speed}
  n2 = Multiply(A=@n1, B=Multiply(A=Constant() {r=-0.02}, B=Time()))
  n3 = Constant2Vector() {g=1, r=0.5}
  n4 = TextureCoordinate()
  n5 = FunctionInput(Preview=Constant()) {bUsePreviewValueAsDefault=true, inputName=IsVehicle, inputType=FunctionInput_Scalar, sortPriority=-19}
  n6 = LinearInterpolate(
   A=LinearInterpolate(
    A=ScalarParameter() {defaultValue=2, group=WeirdObjects, parameterName=Distort-Scale}
    B=ScalarParameter() {defaultValue=7, group=WeirdObjects, parameterName=Distort-Scale-Vehic}
    Alpha=@n5
   )
   B=ScalarParameter() {defaultValue=2, group=WeirdObjects, parameterName=Distort-Scale_Char}
   Alpha=NamedRerouteUsage()
  )
  n7 = TextureSample(UVs=Add(A=Multiply(A=@n2, B=@n3), B=Multiply(A=@n4, B=@n6))) {samplerType=SAMPLERTYPE_Grayscale, texture=&LowResBlurredNoise}
  ConstantBiasScale(None=@n7.RGB) {bias=-0.5, scale=2}
  Constant() {r=25}
  n8 = SetMaterialAttributes!(MaterialAttributes, BaseColor=Constant3Vector() {constant=<rgba 0.3438 0.2227 0.0243 0>}, Metallic, Specular, Roughness=Constant() {r=0.4}) {attributeSetTypes [
    #1
    #2
    #3
    #4
   ]}
  n9 = FunctionInput(Preview=@n8) {inputName=Material, inputType=FunctionInput_MaterialAttributes, sortPriority=-200}
  n10 = BreakMaterialAttributes!(Attr=@n9)
  n11 = Power(Base=@n10.BaseColor, Exp=ScalarParameter() {defaultValue=0.4, group=WeirdObjects, parameterName=BaseColorInput-Pow})
  n12 = ScalarParameter() {defaultValue=0.5, group=WeirdObjects, parameterName=BaseColorClampMax}
  n13 = TextureObjectParameter() {
   desc=Roughness in Alpha Channel
   group=WeirdObjects
   parameterName=WeirdObjects_Mask
   texture=&T_BlacksRGB
  }
  n14 = TextureObjectParameter() {
   desc=Roughness in Alpha Channel
   group=WeirdObjects
   parameterName=WeirdObjects_Normal
   samplerType=SAMPLERTYPE_Normal
   texture=&FlatNormalBlue
  }
  n15 = FunctionInput(Preview=VertexInterpolator(VS=PreSkinnedPosition()).PS) {bUsePreviewValueAsDefault=true, inputName=PreSkinLocalPosition}
  n16 = NamedRerouteUsage()
  n17 = LinearInterpolate(
   A=LinearInterpolate(
    A=ScalarParameter() {defaultValue=8, group=WeirdObjects, parameterName=Distort-Amount}
    B=ScalarParameter() {defaultValue=40, group=WeirdObjects, parameterName=Distort-Amount-Vehic}
    Alpha=@n5
   )
   B=ScalarParameter() {group=WeirdObjects, parameterName=Distort-Amount-Char}
   Alpha=@n16
  )
  n18 = Constant() {r=0.05}
  n19 = Time()
  n20 = Add(A=Multiply(A=Multiply(A=@n18, B=@n19), B=Constant2Vector() {g=0.3, r=1}), B=Multiply(A=TextureCoordinate(), B=Constant() {r=6}))
  n21 = Multiply(
   A=@n7.R
   B=Power(
    Base=TextureSample(UVs=@n20) {samplerType=SAMPLERTYPE_Grayscale, texture=&LowResBlurredNoise}.RGB
    Exp=Constant() {r=2.5}
   )
  )
  n22 = ScalarParameter() {defaultValue=25, group=WeirdObjects, parameterName=Scale-Master}
  n23 = FunctionInput(Preview=Constant() {r=1}) {
   bUsePreviewValueAsDefault=true
   desc=VEHICLE wrap scale
   inputName=Wrap ScaleMultiplier
   inputType=FunctionInput_Scalar
   previewValue=<1 0 0 1>
  }
  n24 = Add(A=ComponentMask(None=NamedRerouteUsage()) {r=true}, B=@n16)
  n25 = PreSkinnedNormal()
  n26 = MaterialFunctionCall!(
   Material Normals
   TextureObject_MASK=@n13
   TextureObject_Diffuse
   TextureObject_Normal=@n14
   TextureScale=LinearInterpolate(
    A=LinearInterpolate(A=@n22, B=@n23, Alpha=@n5)
    B=ScalarParameter() {defaultValue=25, group=WeirdObjects, parameterName=Scale-Master_Character}
    Alpha=Saturate(None=@n24)
   )
   Normal Texture Intensity=ScalarParameter() {defaultValue=2.5, group=WeirdObjects, parameterName=VeinsNormalIntensity, sortPriority=-17}
   Pre-Skinned Local Position=Add(A=@n15, B=Multiply(A=@n17, B=@n21)) {constB=0}
   Pre-Skinned Local Normal=FunctionInput(Preview=VertexInterpolator(VS=@n25).PS) {bUsePreviewValueAsDefault=true, inputName=PreSkinnedLocalNormal}
   Rotation Z Axis
   Rotation X Axis
   Triplanner_Offset
   isMech
   Export Float 4
   ProjectionTransitionContrast=Constant() {r=1}
   IsStatsB
   isCharacter
  ) {materialFunction=/Game/Packages/Fortress_SharedMaterials/Base_Material_Functions/BaseCustomizationFunctions/MF_Base_Customization_Triplanar}
  n27 = ComponentMask(None=@n26.Output MASK) {r=true}
  n28 = Multiply(
   A=Divide(
    A=ObjectRadius()
    B=MaterialFunctionCall() {materialFunction=/Engine/Functions/Engine_MaterialFunctions02/WorldPositionOffset/ObjectScale}.Scale X
   )
   B=Constant() {r=1.75}
  )
  n29 = Add(A=LinearInterpolate(A=Constant(), B=Constant() {r=0.15}, Alpha=@n5), B=ConstantBiasScale(None=Divide(A=@n15, B=@n28)))
  n30 = Constant3Vector() {constant=<rgba 0 1 0 0>}
  n31 = MaterialFunctionCall(Float3=Multiply(A=Saturate(None=@n29), B=FunctionInput(Preview=@n30) {inputName=ForwardAxis})) {materialFunction=&BreakOutFloat3Components}
  n32 = Max(A=Max(A=@n31.R, B=@n31.R), B=@n31.B)
  n33 = ScalarParameter() {defaultValue=3, group=WeirdObjects, parameterName=BurnAwayMaskPower}
  n34 = OneMinus(None=@n21)
  n35 = MaterialFunctionCall(
   A=@n27
   B=Power(Base=@n27, Exp=Constant() {r=2})
   Transition Phase=Multiply(A=Saturate(None=@n32), B=Power(Base=@n34, Exp=@n33))
   Height Texture=@n27
   Contrast=Constant() {r=0.5}
  ) {materialFunction=&HeightLerp}
  n36 = Power(Base=@n35.Results, Exp=Constant() {r=2})
  n37 = Power(Base=@n34, Exp=ScalarParameter() {defaultValue=30, group=WeirdObjects, parameterName=ColorMaskPower})
  n38 = LinearInterpolate(
   A=VectorParameter() {group=WeirdObjects, parameterName=VeinsColor, defaultValue=<rgba 0.0309 0.05 0.0325 1>}.RGB
   B=VectorParameter() {group=WeirdObjects, parameterName=VeinsColor2, defaultValue=<rgba 0.0625 0.0396 0.0247 1>}.RGB
   Alpha=@n37
  )
  n39 = VectorParameter() {group=WeirdObjects, parameterName=VeinNormalAmount, defaultValue=<rgba 1 1 1 1>}
  n40 = Normalize(VectorInput=LinearInterpolate(
    A=GetMaterialAttributes(None=@n9) {attributeGetTypes [#5, #4, #3]}.Normal
    B=Multiply(A=@n26.Normal Texture, B=@n39.RGB)
    Alpha=@n27
   ))
  n41 = DotProduct(A=Transform(None=@n40), B=CameraVectorWS())
  n42 = Power(Base=OneMinus(None=Saturate(None=@n41)), Exp=Constant() {r=2})
  n43 = LinearInterpolate(
   A=VectorParameter() {group=WeirdObjects, parameterName=DiffuseFresnelColor, defaultValue=<rgba 0.4062 0.39 0.3071 1>}.RGB
   B=VectorParameter() {group=WeirdObjects, parameterName=DiffuseFresnelColor2, defaultValue=<rgba 0.2 0.1455 0.1086 1>}.RGB
   Alpha=@n37
  )
  n44 = LinearInterpolate(
   A=Multiply(
    A=ScalarParameter() {defaultValue=0.5, group=WeirdObjects, parameterName=BaseColorMaximumBrightness}
    B=Clamp(None=@n11, Max=@n12) {clampMode=CMODE_ClampMax}
   )
   B=Add(A=Multiply(A=@n36, B=@n38), B=Multiply(A=@n42, B=@n43))
   Alpha=@n35.Alpha
  ) {desc=Diffuse}
  n45 = LinearInterpolate(A=@n10.Metallic, B=Constant(), Alpha=@n35.Alpha) {desc=Metal}
  n46 = Constant() {r=3}
  n47 = Saturate(None=Add(A=Multiply(A=@n35.Alpha, B=@n46), B=Constant() {r=-1}))
  n48 = LinearInterpolate(A=@n10.Specular, B=ScalarParameter() {defaultValue=0.3, group=WeirdObjects, parameterName=VeinSpec}, Alpha=@n47) {desc=Spec}
  n49 = LinearInterpolate(A=@n10.Roughness, B=ScalarParameter() {defaultValue=0.6, group=WeirdObjects, parameterName=VeinRoughness}, Alpha=@n47) {desc=Roughness}
  n50 = LinearInterpolate(A=@n10.Normal, B=@n40, Alpha=@n35.Alpha) {desc=Normals}
  FunctionOutput(None=StaticSwitchParameter(
    True=MakeMaterialAttributes!(
     BaseColor=@n44
     Metallic=@n45
     Specular=@n48
     Roughness=@n49
     Anisotropy
     EmissiveColor=@n10.EmissiveColor
     Opacity=@n10.EmissiveColor
     OpacityMask=@n10.EmissiveColor
     Normal=@n50
     Tangent
     WorldPositionOffset=@n10.EmissiveColor
     SubsurfaceColor=@n10.EmissiveColor
     ClearCoat=@n10.EmissiveColor
     ClearCoatRoughness=@n10.EmissiveColor
     AmbientOcclusion=@n10.EmissiveColor
     Refraction=@n10.EmissiveColor
     CustomizedUVs_0=@n10.EmissiveColor
     CustomizedUVs_1=@n10.EmissiveColor
     CustomizedUVs_2=@n10.EmissiveColor
     CustomizedUVs_3=@n10.EmissiveColor
     CustomizedUVs_4=@n10.EmissiveColor
     CustomizedUVs_5=@n10.EmissiveColor
     CustomizedUVs_6=@n10.EmissiveColor
     CustomizedUVs_7=@n10.EmissiveColor
     PixelDepthOffset=@n10.EmissiveColor
     ShadingModel=@n10.EmissiveColor
     Displacement
    )
    False=@n9
   ) {group=Customization_Wrap, parameterName=Weird Objects Wrap})
  LinearInterpolate(A=@n1, B=ScalarParameter() {defaultValue=1, group=WeirdObjects, parameterName=Distort-Speed-Vehic}, Alpha=@n5)
  Multiply(A=@n35.Alpha)
  Multiply()
  n51 = FunctionInput() {bUsePreviewValueAsDefault=true, inputName=IsStatsB, inputType=FunctionInput_Vector2, sortPriority=-19, previewValue=<1 0 0 1>}
  NamedRerouteDeclaration(None=@n51) {name=IsStatsB, nodeColor=<rgba 0 1 0.5313 1>}
  n52 = FunctionInput() {bUsePreviewValueAsDefault=true, inputName=IsCharacter, inputType=FunctionInput_Scalar, sortPriority=-19, previewValue=<1 0 0 1>}
  NamedRerouteDeclaration(None=@n52) {name=IsCharacter, nodeColor=<rgba 0 0.9297 1 1>}
  ShadingPathSwitch(Default=@n40, Mobile=Constant3Vector() {constant=<rgba 0 0 1 0>}) {desc=Migrated from FeatureLevelSwitch}
 ]}
