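# Material function that applies a cosmetic wrap skin to a Material Attributes input, with adjustments for vehicle, mech, character and backpack.
# Shared panning and scale logic drives two triplanar projections of diffuse, normal and mask textures.
# A directional gradient, optionally height-blended by the first layer, mixes them and feeds masked emissive, with optional pulse tinting.
# A time-of-day emissive pass follows, then the result blends over the input by the customization mask behind a static switch.
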MaterialFunction /Game/Weapons/WeaponSkins/Wraps/MaterialFunctions/MF_Customization_Wrap
graph {nodes [
  n1 = PreSkinnedLocalBounds()
  n2 = FunctionInput() {bUsePreviewValueAsDefault=true, inputName=Pre-Skinned Local Position, sortPriority=41}
  n3 = Subtract(A=@n1.Max, B=@n1.Max)
  Divide(A=Subtract(A=@n2, B=@n1.Min), B=@n3)
  FunctionInput(Preview=Constant()) {bUsePreviewValueAsDefault=true, inputName=------------------, inputType=FunctionInput_Scalar, sortPriority=-19}
  FunctionInput(Preview=Constant()) {bUsePreviewValueAsDefault=true, inputName=-------------------, inputType=FunctionInput_Scalar, sortPriority=23}
  n4 = FunctionInput() {inputName=isVehicle, inputType=FunctionInput_Scalar}
  n5 = FunctionInput(Preview=Constant3Vector() {constant=<rgba 1 0 0 0>}) {inputName=ForwardAxis}
  n6 = FunctionInput() {bUsePreviewValueAsDefault=true, inputName=Per Obj Bounds, sortPriority=50}
  n7 = MaterialFunctionCall!(isVehicle=@n4, Pre Skinned Local Position=@n2, ForwardAxis=@n5, Per Obj Bounds=@n6) {desc=Gradient, materialFunction=&/MF_WrapGradient}
  n8 = ScalarParameter() {defaultValue=-2, group=Primary Triplaner Gradient, parameterName=GradientHeightLerp_PatternMin, sortPriority=-13}
  n9 = ScalarParameter() {defaultValue=1, group=Primary Triplaner Gradient, parameterName=GradientHeightLerp_PatternMax, sortPriority=-12}
  n10 = VectorParameter() {group=Primary Triplaner DiffusePanningMask, parameterName=Pulse_high, defaultValue=<rgba 1.5464 6.836 10 1>}
  n11 = VectorParameter() {group=Primary Triplaner DiffusePanningMask, parameterName=Pulse_low, defaultValue=<rgba 0.217 1.1037 2 1>}
  n12 = LinearInterpolate(
   A=@n10.RGB
   B=@n11.RGB
   Alpha=MaterialFunctionCall!(
    Brightness-Low=Constant()
    Brightness-High=Constant() {r=1}
    Frequency=ScalarParameter() {defaultValue=0.65, group=Primary Triplaner DiffusePanningMask, parameterName=pulse_time}
    Sinewave-Offset=Constant() {r=0.5}
    Amplitude
    Clamp Sine Wave?
    Clamp-Min
    Clamp-Max
   ) {materialFunction=&/MF_Pulse_Illuminate}.Result
  )
  n13 = TextureObjectParameter() {
   desc="GrayScale For Warping uvs "
   group=Primary Triplaner
   parameterName=Triplaner_01_MASK_Gray
   samplerType=SAMPLERTYPE_LinearGrayscale
   sortPriority=-20
   texture=/Game/Global/Textures/Default/Blanks/T_BlackLinearGrayscale
  }
  n14 = TextureObjectParameter() {
   desc="R = Metal, G = Emissive"
   group=Primary Triplaner
   parameterName=Triplaner_01_MASK
   samplerType=SAMPLERTYPE_Masks
   sortPriority=-20
   texture=&T_BlackMask
  }
  n15 = StaticSwitchParameter(True=@n13, False=@n14) {desc=changes compreassion Mask to Grayscale, group=Primary Triplaner, parameterName=Triplaner_01_MASK Grayscale Compresion, sortPriority=-19}
  n16 = TextureObjectParameter() {
   desc=Roughness in Alpha Channel
   group=Primary Triplaner
   parameterName=Triplaner_01_Diffuse
   sortPriority=-25
   texture=&T_BlacksRGB
  }
  n17 = TextureObjectParameter() {
   group=Primary Triplaner
   parameterName=Triplaner_01_Normals
   samplerType=SAMPLERTYPE_Normal
   sortPriority=-19
   texture=&FlatNormalBlue
  }
  n18 = FunctionInput() {bUsePreviewValueAsDefault=true, inputName=Wrap ScaleMultiplier, inputType=FunctionInput_Scalar, previewValue=<1 0 0 1>}
  n19 = FunctionInput() {inputName=isMech, inputType=FunctionInput_Scalar}
  n20 = MaterialFunctionCall!(
   TextureScale=Multiply(A=ScalarParameter() {defaultValue=1, group=Primary Triplaner, parameterName=Wrap_Texture Scale, sortPriority=-50}, B=@n18)
   isVehicle=@n4
   IsCharacter=NamedRerouteUsage()
   ForwardAxis=@n5
   isMech=@n19
   IsStatsB=NamedRerouteUsage()
   Pre-Skinned Local Position=@n2
   Pre-Skinned Bounds=@n3
  ) {
   desc="Panning and triplaner offset, Texture Scale/ Use Object size "
   materialFunction=&/MF_WrapPanningWorldSpace
  }
  n21 = ScalarParameter() {defaultValue=1, group=Primary Triplaner, parameterName=Triplaners Normal Texture Intensity, sortPriority=-16}
  n22 = FunctionInput() {bUsePreviewValueAsDefault=true, inputName=Pre-Skinned Local Normal, sortPriority=42}
  n23 = StaticBool() {value=true}
  n24 = LinearInterpolate(
   A=ScalarParameter() {defaultValue=1, group=Primary Triplaner, parameterName=Triplaners Seam Hardness, sortPriority=-15}
   B=ScalarParameter() {defaultValue=1, group=Primary Triplaner, parameterName=Triplaners Seam Hardness_Character, sortPriority=-15}
   Alpha=NamedRerouteUsage()
  )
  n25 = ScalarParameter() {
   defaultValue=1
   desc="object normal not the Wrap normals. "
   group=Primary Triplaner
   parameterName=Triplaner_01_Object's_Normal_Strength
   sortPriority=-19
  }
  n26 = FunctionInput() {inputName=Material, inputType=FunctionInput_MaterialAttributes, sortPriority=-200}
  n27 = GetMaterialAttributes(None=@n26) {attributeGetTypes [
    #1
    #2
    #3
    #4
    #5
   ]}
  n28 = LinearInterpolate(A=Constant3Vector() {constant=<rgba 0 0 1 0>}, B=@n27.Normal, Alpha=@n25)
  n29 = FunctionInput() {bUsePreviewValueAsDefault=true, inputName=Rotation X Axis, inputType=FunctionInput_Scalar}
  n30 = ScalarParameter() {group=Primary Triplaner, parameterName=Global Z Rot, sortPriority=-35}
  n31 = LinearInterpolate(
   A=LinearInterpolate(
    A=Add(A=@n29, B=@n30)
    B=ScalarParameter() {group=Primary Triplaner, parameterName=Global Z Rot_Character, sortPriority=-35}
    Alpha=NamedRerouteUsage()
   )
   B=ScalarParameter() {defaultValue=0.25, group=Primary Triplaner, parameterName=Global Z Rot_Backpack, sortPriority=-35}
   Alpha=NamedRerouteUsage()
  )
  n32 = Add(
   A=FunctionInput() {bUsePreviewValueAsDefault=true, inputName=Rotation Z Axis, inputType=FunctionInput_Scalar}
   B=ScalarParameter() {group=Primary Triplaner, parameterName=Global X Rot, sortPriority=-40}
  )
  n33 = ScalarParameter() {group=Primary Triplaner, parameterName=Global X Rot_Character, sortPriority=-35}
  n34 = NamedRerouteUsage()
  n35 = MaterialFunctionCall!(
   Material Normals=@n28
   TextureObject_MASK=@n15
   TextureObject_Diffuse=@n16
   TextureObject_Normal=@n17
   TextureScale=@n20.Triplaner_01_Scale
   Normal Texture Intensity=@n21
   Pre-Skinned Local Position=@n20.Triplaner_01_Panning
   Pre-Skinned Local Normal=@n22
   Rotation Z Axis=LinearInterpolate(
    A=LinearInterpolate(A=@n32, B=@n33, Alpha=@n34)
    B=ScalarParameter() {group=Primary Triplaner, parameterName=Global X Rot_Backpack, sortPriority=-35}
    Alpha=NamedRerouteUsage()
   )
   Rotation X Axis=@n31
   Triplanner_Offset=@n20.Triplaner_01_Scale
   isMech=@n19
   Export Float 4=@n23
   ProjectionTransitionContrast=@n24
   IsStatsB=NamedRerouteUsage()
   isCharacter=NamedRerouteUsage()
  ) {materialFunction=/Game/Packages/Fortress_SharedMaterials/Base_Material_Functions/BaseCustomizationFunctions/MF_Base_Customization_Triplanar}
  n36 = MaterialFunctionCall!(
   Material Normals=@n28
   Diffuse=StaticSwitchParameter(True=Multiply(A=@n12, B=@n35.Diffuse Texture), False=@n35.Diffuse Texture) {group=Primary Triplaner DiffusePanningMask, parameterName=Pulse_Illuminate}
   Roughness=@n35.Roughness Texture
   Normal=@n35.Roughness Texture
  ) {desc=Simple material overider for 1nd triplanner, materialFunction=&/MF_WrapSimpleColor1}
  n37 = MaterialFunctionCall(
   Transition Phase=@n7.Result
   Height Texture=Clamp(None=LinearInterpolate(A=@n8, B=@n9, Alpha=@n36.Roughness))
   Contrast=ScalarParameter() {group=Primary Triplaner Gradient, parameterName=GradientHeightLerpContrast, sortPriority=-11}
  ) {materialFunction=&HeightLerp}
  n38 = StaticBoolParameter() {group=Primary Triplaner Emissive, parameterName="Emissive (Uses Mask for Metal and Emissive)", sortPriority=-1}
  n39 = FunctionInput(Preview=VertexInterpolator(VS=@n1.Half Extents).PS) {bUsePreviewValueAsDefault=true, inputName=Pre-Skinned Bounds, sortPriority=50}
  n40 = Divide(A=Add(A=@n2, B=@n39), B=@n39)
  n41 = StaticSwitchParameter(True=@n6, False=@n40) {
   desc=maps the gradent to Back to front of the weapon.
   group=Primary Triplaner Emissive
   parameterName=Fix-Gradent-To-Match-Bounds
   sortPriority=-9
  }
  n42 = TextureObjectParameter() {
   desc=Roughness in Alpha Channel
   group=Primary Triplaner
   parameterName=Triplaner_02_Diffuse
   sortPriority=-18
   texture=&T_BlacksRGB
  }
  n43 = TextureObjectParameter() {
   group=Primary Triplaner
   parameterName=Triplaner_02_Normals
   samplerType=SAMPLERTYPE_Normal
   sortPriority=-17
   texture=&FlatNormalBlue
  }
  n44 = MaterialFunctionCall(
   Material Normals=@n27.Normal
   TextureObject_Diffuse=@n42
   TextureObject_Normal=@n43
   TextureScale=@n20.Triplaner_02_Scale
   Normal Texture Intensity=@n21
   Pre-Skinned Local Position=@n20.Triplaner_02_Scale
   Pre-Skinned Local Normal=@n22
   Rotation Z Axis=@n32
   Rotation X Axis=@n31
   Triplanner_Offset=@n20.Triplaner_02_Scale
   isMech=@n19
   Export Float 4=@n23
   ProjectionTransitionContrast=@n24
  ) {materialFunction=&MF_Base_Customization_Triplanar}
  n45 = MaterialFunctionCall!(Material Normals=@n27.Normal, Diffuse=@n44.Diffuse Texture, Roughness=@n44.Diffuse Texture, Normal=@n44.Diffuse Texture) {desc=Simple material overider for 2nd triplanner, materialFunction=&/MF_WrapSimpleColor2}
  n46 = StaticBoolParameter() {desc="height Blend ", group=Primary Triplaner Gradient, parameterName="BlendGradientByTriPlaner 01 (Alpha)", sortPriority=-14}
  n47 = StaticSwitchParameter(True=@n6, False=@n40) {
   desc=maps the gradent to Back to front of the weapon.
   group=Primary Triplaner DiffusePanningMask
   parameterName=Fix-Gradent-To-Match-Bounds0
   sortPriority=-20
  }
  n48 = MaterialFunctionCall!(
   BaseNormal=@n27.Normal
   isVehicle=@n4
   Specular=ScalarParameter() {defaultValue=0.5, group=Primary Triplaner, parameterName=Triplaner_Specular, sliderMax=1, sortPriority=-15}
   Material=@n26
   isMech=@n19
   Primary_Diffuse=@n36.Diffuse
   Primary_Roughness=@n36.Diffuse
   Primary_Mask=@n35.Output MASK
   Primary_Normal=@n36.Diffuse
   Primary_emissive=Constant()
   Gradient_Diffuse=@n45.Diffuse
   Gradient_Roughness=@n45.Diffuse
   Gradient_Normal=@n45.Diffuse
   RoughnessAlphaBlend=MaterialFunctionCall!(Mask=StaticSwitch(True=@n37.Alpha, False=@n7.Result, Value=@n46)) {materialFunction=/Game/Packages/Fortress_SharedMaterials/Base_Material_Functions/BaseCustomizationFunctions/ThresholdBlend}.Gradient
   NormalizedLocalPosition=@n47
   Forward Vector=@n5
  ) {materialFunction=/Game/Packages/Fortress_SharedMaterials/Base_Material_Functions/BaseCustomizationFunctions/GradientBlendStyles}
  n49 = MaterialFunctionCall!(
   UseMaskGradent=@n37.Alpha
   UseMaskForEmissive=@n38
   NormalizedLocalPosition=StaticSwitchParameter(True=@n41, False=Divide(A=@n2, B=@n39)) {group=Primary Triplaner Emissive, parameterName=Use-Objects-Full-Lenght, sortPriority=-10}
   ForwardDirection=@n5
   isVehicle=@n4
   HeightMapMask=@n36.Roughness
   EmissiveMask=StaticSwitchParameter(True=ScalarParameter() {defaultValue=1, group=Primary Triplaner Emissive, parameterName=EmissiveScalar}, False=@n35.Output MASK) {group=Primary Triplaner Emissive, parameterName=UseScalarEmissiveMask}
   HeightMapMask_2=@n45.Roughness
   WrapDiffuseMask=@n48.WrapDiffuseMask
   isMech=@n19
   IsCharacter=NamedRerouteUsage()
   IsStatsB=NamedRerouteUsage()
  ) {materialFunction=/Game/Packages/Fortress_SharedMaterials/Base_Material_Functions/BaseCustomizationFunctions/MF_WrapEmissive}
  n50 = StaticSwitchParameter(True=LinearInterpolate(A=@n48.Diffuse, Alpha=@n49.Mask) {constB=0}, False=@n48.Diffuse) {
   desc="fill the backgrond area black so emmisve color shows better. "
   group=Primary Triplaner Emissive
   parameterName=Triplaner01MaskDiffusewithBlack
   sortPriority=22
  }
  n51 = ScalarParameter() {group=Primary Triplaner, parameterName=Triplaner_Metalness, sliderMax=1, sortPriority=-10}
  n52 = StaticSwitch(
   True=@n51
   False=@n35.Output MASK
   Value=StaticBoolParameter() {group=Primary Triplaner, parameterName=Disable Mask for Metal, sortPriority=-11}
  )
  n53 = SetMaterialAttributes!(
   MaterialAttributes=@n26
   BaseColor=@n50
   Metallic=ComponentMask(None=StaticSwitch(True=@n52, False=@n51, Value=@n38)) {r=true}
   Specular=@n48.Spec
   Roughness=LinearInterpolate(
    A=ScalarParameter() {group=Primary Triplaner, parameterName=Triplaner_RoughnessMin, sortPriority=-14}
    B=ScalarParameter() {defaultValue=1, group=Primary Triplaner, parameterName=Triplaner_RoughnessMax, sortPriority=-13}
    Alpha=@n48.Roughness
   )
   Normal=ShadingPathSwitch(Default=@n48.Normal, Deferred=@n48.Normal, Mobile=@n27.Normal) {desc=Migrated from FeatureLevelSwitch}
   EmissiveColor=Add(A=@n49.Result, B=@n48.Emissive)
  ) {attributeSetTypes [
    #2
    #5
    #4
    #3
    #1
    #6
   ]}
  n54 = MaterialFunctionCall!(Material=@n53, Diffuse=@n45.Diffuse, Roghness=@n45.Roughness) {materialFunction=&/MF_TimeofDayEmissiveAdjustments}
  n55 = FunctionInput() {bUsePreviewValueAsDefault=true, inputName="Customization Mask (_CM B)", inputType=FunctionInput_Scalar, sortPriority=24}
  FunctionOutput(None=StaticSwitchParameter(True=BlendMaterialAttributes(A=@n26, B=@n54.Result, Alpha=@n55), False=@n26) {group=Customization, parameterName=Triplanar Texture Wrap})
  FunctionOutput(None=@n35.Output MASK) {outputName=Triplaner Mask}
  FunctionOutput(None=@n35.Roughness Texture) {outputName=Triplaner 01 Alpha}
  FunctionOutput(None=@n44.Roughness Texture) {outputName=Triplaner 02 Alpha}
  n56 = FunctionInput() {bUsePreviewValueAsDefault=true, inputName=IsStatsB, inputType=FunctionInput_Vector2, sortPriority=78}
  NamedRerouteDeclaration(None=@n56) {name=Is_StatsB, nodeColor=<rgba 0 0.9253 1 1>}
  NamedRerouteDeclaration(None=ComponentMask(None=@n56) {r=true}) {name=IsBackpack, nodeColor=<rgba 0 0.9253 1 1>}
  NamedRerouteDeclaration(None=FunctionInput() {bUsePreviewValueAsDefault=true, inputName=isCharacter, inputType=FunctionInput_Scalar, sortPriority=78}) {name=isCharacter, nodeColor=<rgba 0 0.9253 1 1>}
  NamedRerouteUsage()
  NamedRerouteUsage()
 ]}
