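# Material function for the "Date" reactive wrap skin on weapons, vehicles and mechs, enabled by static switch Wrap_Date.
# Top/side projections over wrap bounds supply surface maps and masks, with per-type UV controls.
# Time since a Blueprint-fed timestamp fades a scrolling, warpable second mask, height-lerped with the first into flat "dot" overrides, and a curve-atlas gradient emissive.
# The customization mask blends the result over the incoming material attributes.
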MaterialFunction /Game/Weapons/WeaponSkins/Wraps/MaterialFunctions/MF_Date_Wrap
graph {nodes [
  n1 = VectorParameter() {group=TopSideProjection_UVS, parameterName=TS_Offset_Weapon_02---------------------, sortPriority=160, defaultValue=<rgba 0 0 0 1>}
  n2 = VectorParameter() {group=TopSideProjection_UVS, parameterName=TS_Offset_Vehicle_02, sortPriority=161, defaultValue=<rgba 0 0 0 1>}
  n3 = VectorParameter() {
   desc=must activate TS_Vehicle_Side_02 UVs
   group=TopSideProjection_UVS
   parameterName=TS_Offset_Vehicle_Side_02
   sortPriority=161
   defaultValue=<rgba 0 0 0 1>
  }
  n4 = FunctionInput() {bUsePreviewValueAsDefault=true, inputName=Side Project Vehicle, inputType=FunctionInput_Scalar, sortPriority=3}
  n5 = Multiply(A=@n4, B=ScalarParameter() {group=TopSideProjection_UVS, parameterName="TS_Vehicle_Side_01 UVs 0=Off 1=Active"})
  n6 = LinearInterpolate(A=@n2.RGB, B=@n3.RGB, Alpha=@n5)
  n7 = FunctionInput(Preview=Constant()) {inputName=isVehicle, inputType=FunctionInput_Scalar, sortPriority=1}
  n8 = VectorParameter() {group=TopSideProjection_UVS, parameterName=TS_Offset_Mech_02, sortPriority=162, defaultValue=<rgba 0 0 0 1>}
  n9 = FunctionInput(Preview=Constant()) {bUsePreviewValueAsDefault=true, inputName=isMech, inputType=FunctionInput_Scalar, sortPriority=2}
  n10 = PreSkinnedPosition()
  n11 = MaterialFunctionCall!(
   is Vehicle=@n7
   Object Bounds Max Adjustment=FunctionInput() {bUsePreviewValueAsDefault=true, inputName=Object Bounds Max Adjustment, sortPriority=8}
   Object Bounds Min Adjustment=FunctionInput() {bUsePreviewValueAsDefault=true, inputName=Object Bounds Min Adjustment, sortPriority=7}
   Object Bounds Center Adjustment=FunctionInput() {bUsePreviewValueAsDefault=true, inputName=Object Bounds Center Adjustment, sortPriority=9}
   isMech=@n9
   Side Project Vehicle=@n4
   Local Position=FunctionInput(Preview=VertexInterpolator(VS=@n10).PS) {bUsePreviewValueAsDefault=true, inputName=Local Position, sortPriority=6}
  ) {materialFunction=&/MF_WrapBounds}
  Add(A=@n11.Result, B=LinearInterpolate(A=LinearInterpolate(A=@n1.RGB, B=@n6, Alpha=@n7), B=@n8.RGB, Alpha=@n9))
  n12 = VectorParameter() {group=TopSideProjection, parameterName=TS_Fill_Z_Axis_Weapon_Color_Mask_02, sortPriority=67, defaultValue=<rgba 0 0 0 1>}
  AppendVector(A=@n12.RGB, B=@n12.RGB)
  n13 = FunctionInput() {inputName=Material, inputType=FunctionInput_MaterialAttributes, sortPriority=-10}
  n14 = GetMaterialAttributes(None=@n13) {attributeGetTypes [#1, #2, #3]}
  n15 = ScalarParameter() {defaultValue=1, group=Date_Reactive, parameterName=DA_Dot_BaseColor, sortPriority=19}
  n16 = Constant()
  n17 = Time()
  n18 = ScalarParameter() {defaultValue=-9999999, group=00-Reactive-GTFB, parameterName=00-GameTimeFromBP}
  n19 = Saturate(None=Divide(A=Subtract(A=@n17, B=@n18), B=ScalarParameter() {defaultValue=1, group=Date_Reactive, parameterName=DA_AnimationDuration, sortPriority=33}))
  n20 = Power(Base=@n19, Exp=ScalarParameter() {defaultValue=0.5, group=Date_Reactive, parameterName=DA_DotCurve_Amount, sortPriority=35})
  n21 = ScalarParameter() {group=Date_Reactive, parameterName=DA_TestReactive, sortPriority=300}
  n22 = TextureObjectParameter() {
   group=TopSideProjection
   parameterName=TS_Mask_02
   samplerType=SAMPLERTYPE_Masks
   sortPriority=62
   texture=&T_BlackMask
  }
  n23 = VectorParameter() {group=TopSideProjection, parameterName=TS_Fill_alternate_Axis_Color_Mask 02, sortPriority=67, defaultValue=<rgba 0 0 0 1>}
  n24 = VectorParameter() {group=TopSideProjection_UVS, parameterName=TS-Rotation-Weapon-02---------------------, sortPriority=150, defaultValue=<rgba 0 0 -0.25 1>}
  n25 = VectorParameter() {group=TopSideProjection_UVS, parameterName=TS_Rotation_Vehicle_02, sortPriority=151, defaultValue=<rgba 0 0 0 1>}
  n26 = VectorParameter() {
   desc=must activate TS_Vehicle_Side_02 UVs
   group=TopSideProjection_UVS
   parameterName=TS_Rotation_Vehicle_Side_02
   sortPriority=151
   defaultValue=<rgba 0 0 0 1>
  }
  n27 = LinearInterpolate(A=@n25.RGB, B=@n26.RGB, Alpha=@n5)
  n28 = VectorParameter() {group=TopSideProjection_UVS, parameterName=TS_Rotation_Mech_02, sortPriority=153, defaultValue=<rgba 0.25 0.25 0 1>}
  n29 = VectorParameter() {group=TopSideProjection_UVS, parameterName=TS_Panning_Weapon_02---------------------, sortPriority=180, defaultValue=<rgba 0 0 0 1>}
  n30 = VectorParameter() {group=TopSideProjection_UVS, parameterName=TS_Panning_Vehicle_02, sortPriority=181, defaultValue=<rgba 0 0 0 1>}
  n31 = VectorParameter() {
   desc=must activate TS_Vehicle_Side_02 UVs
   group=TopSideProjection_UVS
   parameterName=TS_Panning_Vehicle_Side_02
   sortPriority=181
   defaultValue=<rgba 0 0 0 1>
  }
  n32 = LinearInterpolate(A=@n29.RGB, B=LinearInterpolate(A=@n30.RGB, B=@n31.RGB, Alpha=@n5), Alpha=@n7)
  n33 = VectorParameter() {group=TopSideProjection_UVS, parameterName=TS_Panning_Mech_02, sortPriority=182, defaultValue=<rgba 0 0 0 1>}
  n34 = Add(A=@n11.Result, B=Multiply(A=Time(), B=LinearInterpolate(A=@n32, B=@n33.RGB, Alpha=@n9)))
  n35 = VectorParameter() {group=TopSideProjection_UVS, parameterName="TS_Scale_Weapon_02 A=uniform", sortPriority=170, defaultValue=<rgba 1 1 1 1>}
  n36 = VectorParameter() {group=TopSideProjection_UVS, parameterName="TS_Scale_Vehicle_02 A=uniform", sortPriority=171, defaultValue=<rgba 1 1 1 1>}
  n37 = Multiply(A=@n36.RGB, B=@n36.RGB)
  n38 = VectorParameter() {
   desc=must activate TS_Vehicle_Side_02 UVs
   group=TopSideProjection_UVS
   parameterName="TS_Scale_Vehicle_Side_02 A=uniform"
   sortPriority=171
   defaultValue=<rgba 1 1 1 1>
  }
  n39 = Multiply(A=@n38.RGB, B=@n38.RGB)
  n40 = VectorParameter() {group=TopSideProjection_UVS, parameterName="TS_Scale_Mech_02 A=uniform", sortPriority=172, defaultValue=<rgba 1 1 1 1>}
  n41 = LinearInterpolate(
   A=LinearInterpolate(A=Multiply(A=@n35.RGB, B=@n35.RGB), B=LinearInterpolate(A=@n37, B=@n39, Alpha=@n5), Alpha=@n7)
   B=Multiply(A=@n40.RGB, B=@n40.RGB)
   Alpha=@n9
  )
  n42 = TextureCoordinate()
  n43 = LinearInterpolate(
   A=ScalarParameter() {defaultValue=1, group=TopSideProjection_UVS, parameterName=TS-WarpUVS-02-TileRate Weapon, sortPriority=203}
   B=ScalarParameter() {defaultValue=1, group=TopSideProjection_UVS, parameterName=TS-WarpUVS-02-TileRate Vehicle, sortPriority=204}
   Alpha=@n7
  )
  n44 = Time() {period=1}
  n45 = VectorParameter() {group=TopSideProjection_UVS, parameterName=TS-WarpUVS-02-Anitmated Weapon, sortPriority=207, defaultValue=<rgba 0 0 0 1>}
  n46 = VectorParameter() {group=TopSideProjection_UVS, parameterName=TS-WarpUVS-02-Anitmated Vehicle, sortPriority=208, defaultValue=<rgba 0 0 0 1>}
  n47 = ComponentMask(None=LinearInterpolate(A=@n45.RGB, B=@n46.RGB, Alpha=@n7)) {g=true, r=true}
  n48 = TextureSampleParameter2D(UVs=Add(A=Multiply(A=@n42, B=@n43), B=Multiply(A=@n44, B=@n47))) {group=TopSideProjection_UVS, parameterName=TS-WarpUVS-02-Texture, sortPriority=202, texture=&T_BlacksRGB}
  n49 = ChannelMaskParameter(None=@n48.RGB) {group=TopSideProjection_UVS, parameterName=TS-WarpUVS-02_Channel, sortPriority=201}
  n50 = LinearInterpolate(
   A=ScalarParameter() {group=TopSideProjection_UVS, parameterName=TS-WarpUVS-02-Amount Weapon, sortPriority=205}
   B=ScalarParameter() {group=TopSideProjection_UVS, parameterName=TS-WarpUVS-02-Amount_Vehicle, sortPriority=206}
   Alpha=@n7
  )
  n51 = FunctionInput(Preview=VertexInterpolator(VS=PreSkinnedNormal()).PS) {bUsePreviewValueAsDefault=true, inputName=Pre-Skinned Local Normal, sortPriority=10}
  n52 = MaterialFunctionCall!(
   IsCharacter
   IsStatsB
   Texture=@n22
   Rotation_Axis=LinearInterpolate(A=LinearInterpolate(A=@n24.RGB, B=@n27, Alpha=@n7), B=@n28.RGB, Alpha=@n9)
   Caculated_Bounds=Add(A=Multiply(A=@n34, B=@n41), B=Multiply(A=@n49.R, B=@n50))
   Pre_Skinned_Normals=@n51
   Blend_Contrast=LinearInterpolate(
    A=ScalarParameter() {defaultValue=1, group=TopSideProjection, parameterName=TS_Blend_Contrast, sortPriority=-1}
    B=ScalarParameter() {defaultValue=2, group=TopSideProjection, parameterName=TS_Blend_Contrast_Mech, sortPriority=-1}
    Alpha=@n9
   )
   Fill_alternate_Axis_Opacity=ScalarParameter() {group=TopSideProjection, parameterName=TS_Fill_alternate_Axis_Opacity 02, sortPriority=51}
   "Fill_alternate_Axis_Color_A=Roughness"=AppendVector(A=@n23.RGB, B=@n23.RGB)
   Is_Vehicle=@n7
   Is_Mech=@n9
   Enable_Fill_Z_Weapon
   Fill_Z_Weapon_opacity
   Fill_Z_Weapon_Color
  ) {materialFunction=&/MF_TopsSideProjection_Base}
  n53 = LinearInterpolate(A=@n16, B=@n52.Combined, Alpha=LinearInterpolate(A=OneMinus(None=@n20) {desc=inverts curve from 0-1 to 1-0}, B=@n21, Alpha=@n21))
  n54 = Multiply(
   A=Clamp(
    None=ChannelMaskParameter(None=@n53) {group=Date_Reactive, parameterName=TS_MaskChannelDot, sortPriority=6}.R
    Max=ScalarParameter() {defaultValue=1, group=Date_Reactive, parameterName=DA_Mask_Gradent_Clamp Max, sortPriority=10}
   ) {maxDefault=0.99, minDefault=0.01}
   B=ScalarParameter() {defaultValue=1, group=Date_Reactive, parameterName=DA_Mask_Gradent_Opacity_CE, sortPriority=11}
  )
  n55 = TextureObjectParameter() {
   group=TopSideProjection
   parameterName=TS_Mask_01
   samplerType=SAMPLERTYPE_Masks
   sortPriority=20
   texture=&T_BlackMask
  }
  n56 = LinearInterpolate(
   A=ScalarParameter() {defaultValue=1, group=TopSideProjection, parameterName=TS_Blend_Contrast, sortPriority=-1}
   B=ScalarParameter() {defaultValue=2, group=TopSideProjection, parameterName=TS_Blend_Contrast_Mech, sortPriority=-1}
   Alpha=@n9
  )
  n57 = ScalarParameter() {group=TopSideProjection, parameterName=TS_Fill_alternate_Axis_Opacity 01, sortPriority=2}
  n58 = VectorParameter() {group=TopSideProjection, parameterName=TS_Fill_alternate_Axis_Color_Mask_01, sortPriority=21, defaultValue=<rgba 0 0 0 1>}
  n59 = VectorParameter() {group=TopSideProjection_UVS, parameterName=TS-Rotation-Weapon-01---------------------, sortPriority=1, defaultValue=<rgba 0 0 -0.25 1>}
  n60 = VectorParameter() {group=TopSideProjection_UVS, parameterName=TS_Rotation_Vehicle_01, sortPriority=2, defaultValue=<rgba 0 0 0 1>}
  n61 = VectorParameter() {
   desc=must activate TS_Vehicle_Side_01 UVs
   group=TopSideProjection_UVS
   parameterName=TS_Rotation_Vehicle_Side_01
   sortPriority=2
   defaultValue=<rgba 0 0 0 1>
  }
  n62 = VectorParameter() {group=TopSideProjection_UVS, parameterName=TS_Rotation_Mech_01, sortPriority=3, defaultValue=<rgba 0.25 0.25 0 1>}
  n63 = LinearInterpolate(A=LinearInterpolate(A=@n59.RGB, B=LinearInterpolate(A=@n60.RGB, B=@n61.RGB, Alpha=@n5), Alpha=@n7), B=@n62.RGB, Alpha=@n9)
  n64 = VectorParameter() {group=TopSideProjection_UVS, parameterName=TS_Offset_Weapon_01---------------------, sortPriority=10, defaultValue=<rgba 0 0 0 1>}
  n65 = VectorParameter() {group=TopSideProjection_UVS, parameterName=TS_Offset_Vehicle_01, sortPriority=11, defaultValue=<rgba 0 0 0 1>}
  n66 = VectorParameter() {
   desc=must activate TS_Vehicle_Side_01 UVs
   group=TopSideProjection_UVS
   parameterName=TS_Offset_Vehicle_Side_01
   sortPriority=11
   defaultValue=<rgba 0 0 0 1>
  }
  n67 = LinearInterpolate(A=@n64.RGB, B=LinearInterpolate(A=@n65.RGB, B=@n66.RGB, Alpha=@n5), Alpha=@n7)
  n68 = VectorParameter() {group=TopSideProjection_UVS, parameterName=TS_Offset_Mech_01, sortPriority=12, defaultValue=<rgba 0 0 0 1>}
  n69 = VectorParameter() {group=TopSideProjection_UVS, parameterName="TS_Scale_Weapon_01 A=uniform", sortPriority=30, defaultValue=<rgba 1 1 1 1>}
  n70 = Multiply(A=@n69.RGB, B=@n69.RGB)
  n71 = VectorParameter() {group=TopSideProjection_UVS, parameterName="TS_Scale_Vehicle_01 A=uniform", sortPriority=31, defaultValue=<rgba 1 1 1 1>}
  n72 = VectorParameter() {
   desc=must activate TS_Vehicle_Side_01 UVs
   group=TopSideProjection_UVS
   parameterName="TS_Scale_Vehicle_Side_01 A=uniform"
   sortPriority=31
   defaultValue=<rgba 1 1 1 1>
  }
  n73 = LinearInterpolate(A=Multiply(A=@n71.RGB, B=@n71.RGB), B=Multiply(A=@n72.RGB, B=@n72.RGB), Alpha=@n5)
  n74 = VectorParameter() {group=TopSideProjection_UVS, parameterName="TS_Scale_Mech_01 A=uniform", defaultValue=<rgba 1 1 1 1>}
  n75 = Multiply(
   A=Add(A=@n11.Result, B=LinearInterpolate(A=@n67, B=@n68.RGB, Alpha=@n9))
   B=LinearInterpolate(A=LinearInterpolate(A=@n70, B=@n73, Alpha=@n7), B=Multiply(A=@n74.RGB, B=@n74.RGB), Alpha=@n9)
  )
  n76 = MaterialFunctionCall(
   Texture=@n55
   Rotation_Axis=@n63
   Caculated_Bounds=@n75
   Pre_Skinned_Normals=@n51
   Blend_Contrast=@n56
   Fill_alternate_Axis_Opacity=@n57
   "Fill_alternate_Axis_Color_A=Roughness"=AppendVector(A=@n58.RGB, B=@n58.RGB)
   Is_Vehicle=@n7
   Is_Mech=@n9
  ) {materialFunction=&MF_TopsSideProjection_Base}
  n77 = ChannelMaskParameter(None=@n76.Combined) {group=Date_Reactive, maskChannel=Green, parameterName=TS_MaskChannelDot_Height, sortPriority=6, defaultValue=<rgba 0 1 0 0>}
  n78 = Saturate(None=MaterialFunctionCall(
    Transition Phase=Saturate(None=@n54)
    Height Texture=@n77.R
    Contrast=ScalarParameter() {defaultValue=1, group=Date_Reactive, parameterName=DA_Mask_HeightLerp_Contrast, sortPriority=13}
   ) {materialFunction=&HeightLerp}.Alpha)
  n79 = TextureObjectParameter() {group=TopSideProjection, parameterName=TS_BaseColor_01, sortPriority=11, texture=&T_BlacksRGB}
  n80 = VectorParameter() {group=TopSideProjection, parameterName="TS_Fill_alternate_Axis_Color_A=Roughness 01", sortPriority=3}
  n81 = MaterialFunctionCall(
   Texture=@n79
   Rotation_Axis=@n63
   Caculated_Bounds=@n75
   Pre_Skinned_Normals=@n51
   Blend_Contrast=@n56
   Fill_alternate_Axis_Opacity=@n57
   "Fill_alternate_Axis_Color_A=Roughness"=AppendVector(A=@n80.RGB, B=@n80.RGB)
   Is_Vehicle=@n7
   Is_Mech=@n9
  ) {materialFunction=&MF_TopsSideProjection_Base}
  n82 = ScalarParameter() {group=TopSideProjection, parameterName=TS_Roughness_Min, sortPriority=13}
  n83 = ScalarParameter() {defaultValue=1, group=TopSideProjection, parameterName=TS_Roughness_Max, sortPriority=14}
  n84 = ComponentMask(None=@n81.Combined) {a=true}
  n85 = TextureObjectParameter() {
   group=TopSideProjection
   parameterName=TS_Normal_01
   samplerType=SAMPLERTYPE_Normal
   sortPriority=16
   texture=&FlatNormalBlue
  }
  n86 = LinearInterpolate(
   A=@n14.Normal
   B=Add(
    A=@n14.Normal
    B=MaterialFunctionCall!(
     Texture_Normals=@n85
     Rotation_Axis=@n63
     Caculated_Bounds=@n75
     Pre_Skinned_Normals=@n51
     Blend_Contrast=@n56
     Fill_alternate_Axis_Opacity=@n57
     Is_Vehicle=@n7
     Is_Mech=@n9
    ) {materialFunction=&/MF_TopsSideProjection_Normals}.Add_Surface_Normal
   )
   Alpha=ScalarParameter() {defaultValue=1, group=TopSideProjection, parameterName=TS_Normal_Intensty, sortPriority=1}
  )
  n87 = ScalarParameter() {group=TopSideProjection, parameterName=TS_Specularty_Min, sortPriority=12}
  n88 = ScalarParameter() {defaultValue=1, group=TopSideProjection, parameterName=TS_Specularty_Max, sortPriority=12}
  n89 = ChannelMaskParameter(None=@n76.Combined) {group=TopSideProjection, parameterName=TS_SpecChannel_Mask_01, sortPriority=26}
  n90 = Multiply(
   A=ChannelMaskParameter(None=@n52.Combined) {group=TopSideProjection_Emissive, parameterName=TS_WarpEmissive01MaskChannel_02, sortPriority=37}.R
   B=ScalarParameter() {group=TopSideProjection_Emissive, parameterName=TS_WarpEmissive01_Warp_Amount, sortPriority=38}
  )
  n91 = LinearInterpolate(
   A=ScalarParameter() {group=TopSideProjection_Emissive, parameterName=TS_Emissive01Curve_Offset_Weapon, sortPriority=4}
   B=ScalarParameter() {group=TopSideProjection_Emissive, parameterName=TS_Emissive01Curve_Offset_Vehicle, sortPriority=4}
   Alpha=@n7
  )
  n92 = ScalarParameter() {group=TopSideProjection_Emissive, parameterName=TS_Emissive01Curve_Offset_Mech, sortPriority=4}
  n93 = ScalarParameter() {
   defaultValue=1
   desc="negative value inverts "
   group=TopSideProjection_Emissive
   parameterName=TS_Emissive01Curve_Falloff_Weapon
   sortPriority=4
  }
  n94 = ScalarParameter() {
   defaultValue=1
   desc="negative value inverts "
   group=TopSideProjection_Emissive
   parameterName=TS_Emissive01Curve_Falloff_Vehicle
   sortPriority=4
  }
  n95 = LinearInterpolate(A=@n93, B=@n94, Alpha=@n7)
  n96 = ScalarParameter() {
   defaultValue=1
   desc="negative value inverts "
   group=TopSideProjection_Emissive
   parameterName=TS_Emissive01Curve_Falloff_Mech
   sortPriority=4
  }
  n97 = Constant3Vector() {constant=<rgba 0 1 0 0>}
  n98 = ScalarParameter() {group=Date_Reactive, parameterName=DA_EmissiveRollRotate, sortPriority=38}
  n99 = FunctionInput(Preview=Constant3Vector() {constant=<rgba 1 0 0 0>}) {bUsePreviewValueAsDefault=true, inputName=ForwardAxis, sortPriority=4}
  n100 = Frac(None=MaterialFunctionCall!(
    Bounds=@n11.Result
    ForwardAxis=LinearInterpolate(A=@n99, B=@n97, Alpha=@n98)
    Offset=LinearInterpolate(A=@n91, B=@n92, Alpha=@n9)
    FallOff=LinearInterpolate(A=@n95, B=@n96, Alpha=@n9)
   ) {materialFunction=&/MF_Wrap_Gradent_Simple}.Raw_Data)
  n101 = CurveAtlasRowParameter(CurveTime=Frac(None=Add(A=@n90, B=@n100))) {
   atlas=/Game/Weapons/WeaponSkins/Wraps/Textures/CruveAtlas_Wraps_01
   curve=/Game/Weapons/WeaponSkins/Wraps/Textures/C_GradSimple
   group=TopSideProjection_Emissive
   parameterName=TS_Emissive01_Curve
   sortPriority=3
  }
  n102 = MaterialFunctionCall!(
   Index=ScalarParameter() {group=TopSideProjection_Emissive, parameterName="TS_Emissive01_Curve  Index", sortPriority=3}
   CCinput=@n101
   CCinput_Alpha=@n101
   UV
  ) {materialFunction=&/MF_Global_CC_Tex}
  n103 = Multiply(
   A=@n102.3V
   B=Multiply(
    A=@n102.Alpha
    B=ScalarParameter() {defaultValue=1, group=TopSideProjection_Emissive, parameterName=TS-Emissive01-CurveBrightness, sortPriority=4}
   )
  )
  n104 = ScalarParameter() {defaultValue=0.5, group=Date_Reactive, parameterName=DA_Curve_Amount, sortPriority=34}
  n105 = LinearInterpolate(A=OneMinus(None=Power(Base=@n19, Exp=@n104)) {desc=inverts curve from 0-1 to 1-0}, B=@n21, Alpha=@n21)
  n106 = SetMaterialAttributes!(
   MaterialAttributes=@n14.MaterialAttributes
   BaseColor=LinearInterpolate(A=ComponentMask(None=@n81.Combined) {b=true, g=true, r=true}, B=@n15, Alpha=@n78) {desc=basecolor}
   Roughness=LinearInterpolate(
    A=LinearInterpolate(A=@n82, B=@n83, Alpha=@n84)
    B=ScalarParameter() {defaultValue=1, group=Date_Reactive, parameterName=DA_Dot_Roughness, sortPriority=20}
    Alpha=@n78
   ) {desc=Roughness}
   Normal=LinearInterpolate(
    A=ShadingPathSwitch(Default=@n86, Deferred=@n86, Mobile=@n14.Normal) {desc=Migrated from FeatureLevelSwitch}
    B=Constant3Vector() {constant=<rgba 0 0 1 0>}
    Alpha=@n78
   ) {desc=Normal}
   Specular=LinearInterpolate(A=LinearInterpolate(A=@n87, B=@n88, Alpha=@n89.R), B=@n15, Alpha=@n78) {desc=Spec}
   Metallic=LinearInterpolate(
    A=ChannelMaskParameter(None=@n76.Combined) {group=TopSideProjection, parameterName=TS_MetalicChannel_Mask_01, sortPriority=26}.R
    B=@n15
    Alpha=@n78
   ) {desc=Metallic}
   EmissiveColor=LinearInterpolate(A=LinearInterpolate(A=@n16, B=@n103, Alpha=@n105), B=@n15, Alpha=@n78) {desc=emissive}
  ) {attributeSetTypes [
    #1
    #2
    #3
    #4
    #5
    #6
   ]}
  n107 = Add(
   A=FunctionInput() {bUsePreviewValueAsDefault=true, inputName="Customization Mask (_CM B)", inputType=FunctionInput_Scalar, sortPriority=24}
   B=ScalarParameter() {group=TopSideProjection, parameterName=TS_Remove_Customization Mask, sortPriority=-10}
  )
  n108 = StaticSwitchParameter(True=BlendMaterialAttributes(A=@n14.MaterialAttributes, B=@n106, Alpha=Saturate(None=@n107)), False=@n13) {
   desc="Topside Projection Base - Reactive - Topside Emissive  - Height Lerp based Masked sections (Material exposed)"
   group=Customization_Wrap
   parameterName=Wrap_Date
  }
  FunctionOutput(None=@n108)
  ScalarParameter() {group=TopSideProjection, parameterName=TS_Fill_Z_Axis_Weapon_Opacity 02, sortPriority=51}
 ]}
