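# Weapon wrap material function that adds an optional animated liquid-metal effect to an incoming material, blended through mask inputs.
# A panning projected warp texture distorts two scrolling projected normal layers, which add onto the root normal.
# Fresnel lookups into a spectrum texture or curve atlas tint the base color like thin film, alongside metallic and roughness overrides.
# IsState inputs select projection scale and rotation for weapons, vehicles, mechs, characters, backpacks and pickaxes.
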MaterialFunction /Game/Weapons/WeaponSkins/Wraps/MaterialFunctions/MF_EffectLiquidMetal
graph {nodes [
  n1 = Multiply(A=Time(), B=ScalarParameter() {defaultValue=5, group=EffectLiquidMetal, parameterName=ELMetal_Noise_speed})
  n2 = Multiply(A=NamedRerouteUsage(), B=ScalarParameter() {defaultValue=1, group=EffectLiquidMetal, parameterName=ELMetal_Noise_tile})
  ComponentMask(None=Noise!(World Position=Add(A=@n1, B=@n2), FilterWidth) {bTurbulence=false, levels=1, noiseFunction=NOISEFUNCTION_GradientALU, outputMin=0, scale=0.04}) {g=true}
  n3 = NamedRerouteUsage()
  n4 = NamedRerouteUsage()
  n5 = TextureSample() {samplerType=SAMPLERTYPE_Masks, texture=/Game/Weapons/FORT_Rifles/Textures/T_Scar_CM}
  n6 = FunctionInput(Preview=@n5.R) {bUsePreviewValueAsDefault=true, inputName=Mask Primary, inputType=FunctionInput_Scalar, sortPriority=24, previewValue=<1 0 0 1>}
  n7 = FunctionInput(Preview=@n5.B) {bUsePreviewValueAsDefault=true, inputName=Mask Secondary, inputType=FunctionInput_Scalar, sortPriority=24, previewValue=<1 0 0 1>}
  n8 = Add(
   A=Multiply(A=ScalarParameter() {defaultValue=1, group=EffectLiquidMetal, parameterName="ELMetal__Mask_Primary 1=0n", sortPriority=5}, B=@n6)
   B=Multiply(A=@n7, B=ScalarParameter() {group=EffectLiquidMetal, parameterName="ELMetal__Mask_Accent 1=0n", sortPriority=5})
  )
  n9 = ScalarParameter() {group=EffectLiquidMetal, parameterName="ELMetal__Mask Disable 1=0n", sortPriority=5}
  n10 = FunctionInput() {inputName=TS_Mask_02, inputType=FunctionInput_Vector4, sortPriority=50}
  n11 = FunctionInput() {inputName=TS_Mask_01, inputType=FunctionInput_Vector4, sortPriority=50}
  n12 = TextureSample() {samplerType=SAMPLERTYPE_Normal, texture=/Game/Weapons/FORT_Rifles/T_SCAR_N}
  n13 = Constant3Vector() {constant=<rgba 0.9333 0.957 1 0>}
  n14 = GetMaterialAttributes(None=FunctionInput(Preview=SetMaterialAttributes!(MaterialAttributes, Normal=@n12.RGB, BaseColor=@n13) {attributeSetTypes [#1, #2]}) {inputName=Material, inputType=FunctionInput_MaterialAttributes}) {attributeGetTypes [
    #2
    #3
    #1
    #4
   ]}
  n15 = StaticSwitchParameter(
   True=StaticSwitchParameter(
    True=ChannelMaskParameter(None=@n10) {group=EffectLiquidMetal, parameterName=ELMetal__Mask _TS_Mask_02_Channels, sortPriority=5}.R
    False=ChannelMaskParameter(None=@n11) {group=EffectLiquidMetal, parameterName=ELMetal__Mask _TS_Mask_01_Channels, sortPriority=5}.R
   ) {group=EffectLiquidMetal, parameterName="ELMetal__Mask _TS_Mask_01-------->TS_Mask_02", sortPriority=1}
   False=Add(
    A=@n14.Metallic
    B=ScalarParameter() {defaultValue=1, group=EffectLiquidMetal, parameterName="ELMetal__Mask use Metalic 0=0n", sortPriority=5}
   )
  ) {
   desc="Only works with Topside masks. "
   group=EffectLiquidMetal
   parameterName="ELMetal__Mask_use Metalic-------->TS_Mask_01"
   sortPriority=1
  }
  n16 = Saturate(None=Multiply(A=Add(A=@n8, B=@n9), B=@n15))
  FunctionOutput(None=StaticSwitchParameter(
    True=BlendMaterialAttributes(A=@n3, B=@n4, Alpha=@n16) {vertexAttributeBlendType=UseA}
    False=NamedRerouteDeclaration(None=@n14.MaterialAttributes) {name=Root_Material, nodeColor=<rgba 0.7812 1 0 1>}
   ) {group=Customization, parameterName=Effect_LiquidMetal})
  n17 = Multiply(
   A=MaterialFunctionCall() {materialFunction=&CameraDirectionVector}.Result
   B=ScalarParameter() {defaultValue=1, group=EffectLiquidMetal, parameterName=ELMetal_FlimColor_Camera_Movement}
  )
  n18 = ScalarParameter() {defaultValue=3, group=EffectLiquidMetal, parameterName=ELMetal_FlimColor_Fresnel_Exponenth}
  n19 = ScalarParameter() {defaultValue=0.041, group=EffectLiquidMetal, parameterName=ELMetal_FlimColor_Fresnel_fract}
  n20 = Transform(None=NamedRerouteUsage())
  n21 = Multiply(
   A=Add(
    A=ScalarParameter() {group=EffectLiquidMetal, parameterName=ELMetal_FlimColor_Offset}
    B=Fresnel(ExponentIn=@n18, BaseReflectFractionIn=@n19, Normal=@n20)
   ) {constA=0.5}
   B=ScalarParameter() {defaultValue=1, group=EffectLiquidMetal, parameterName=ELMetal_FlimColor_tileRate}
  )
  n22 = StaticSwitchParameter(True=Add(A=ComponentMask(None=@n17) {g=true}, B=@n21), False=@n21) {group=EffectLiquidMetal, parameterName=ELMetal_FlimColor_Add_Camera_offset}
  n23 = TextureSampleParameter2D(UVs=@n22) {group=EffectLiquidMetal, parameterName=ELMetal_FlimColor_Tex, texture=/Game/Global/Textures/Default/Thinfilm/T_ThinFilm_Spectrum_COLOR}
  n24 = Frac(None=@n22)
  n25 = CurveAtlasRowParameter(CurveTime=@n24) {
   atlas=/Game/Weapons/WeaponSkins/Wraps/Textures/CruveAtlas_Wraps_01
   curve=/Game/Weapons/WeaponSkins/Wraps/Textures/0_ColorCurves/CC_Spectum
   defaultValue=98
   group=EffectLiquidMetal
   parameterName=ELMetal_FlimColor_CC
  }
  n26 = StaticSwitchParameter(
   True=@n23.RGB
   False=MaterialFunctionCall!(
    Index=ScalarParameter() {defaultValue=98, group=EffectLiquidMetal, parameterName="ELMetal_FlimColor_CC  Index"}
    CCinput=@n25
    CCinput_Alpha
    UV=@n24
   ) {materialFunction=&/MF_Global_CC_Tex}.3V
  ) {group=EffectLiquidMetal, parameterName="ELMetal_FlimColor_CC--->Tex"}
  n27 = ScalarParameter() {defaultValue=0.85, group=EffectLiquidMetal, parameterName=ELMetal_FlimColor_Color_Strength}
  n28 = NamedRerouteUsage()
  n29 = NamedRerouteUsage()
  n30 = NamedRerouteDeclaration(None=StaticBoolParameter() {group=EffectLiquidMetal_BC, parameterName=ELMetal__BC_Use_Tex}) {name=BC_Switch, nodeColor=<rgba 0 0.0859 1 1>}
  n31 = Saturate(None=Multiply(
    A=StaticSwitch(True=@n28, False=@n29, Value=@n30)
    B=ScalarParameter() {defaultValue=1, group=EffectLiquidMetal, parameterName=ELMetal_BaseColor_Brightness}
   ))
  n32 = Multiply(
   A=Desaturation(
    None=LinearInterpolate(A=@n26, Alpha=@n27)
    Fraction=ScalarParameter() {defaultValue=0.7, group=EffectLiquidMetal, parameterName=ELMetal_FlimColor_Color_Saturation}
   )
   B=@n31
  ) {constB=0.05}
  Multiply(A=@n32)
  n33 = VectorParameter() {group=EffectLiquidMetal, parameterName=ELMetal_Warp_Alternate}
  n34 = Time()
  n35 = VectorParameter() {group=EffectLiquidMetal, parameterName=ELMetal_Warp_Speed, defaultValue=<rgba 0.025 -0.0133 0 1>}
  n36 = NamedRerouteUsage()
  n37 = ChannelMaskParameter(None=MaterialFunctionCall!(
    Texture=NamedRerouteUsage()
    Caculated_Bounds=NamedRerouteUsage()
    Pre_Skinned_Normals=NamedRerouteUsage()
    Blend_Contrast=ScalarParameter() {defaultValue=1, group=EffectLiquidMetal, parameterName=ELMetal_Contrast}
    IsStatsB=NamedRerouteUsage()
    IsState=NamedRerouteUsage()
    Fill_alternate_Axis_Opacity=Constant() {r=1}
    "Fill_alternate_Axis_Color_A=Roughness"=AppendVector(A=@n33.RGB, B=@n33.RGB)
    Offset=Multiply(A=@n34, B=@n35.RGB)
    Scale=@n36
    Offset_Pivot
    Rotation_Axis=NamedRerouteUsage()
   ) {materialFunction=&/MF_TopsSideProjection_Base_V2}.Combined) {group=EffectLiquidMetal, parameterName=ELMetal_Warp_Channel, sortPriority=5}
  Multiply(A=@n37.R)
  n38 = Power(Base=NamedRerouteUsage(), Exp=ScalarParameter() {defaultValue=1.5, group=EffectLiquidMetal, parameterName=ELMetal_FlimColor_Gmap_Power})
  n39 = ScalarParameter() {defaultValue=0.5, group=EffectLiquidMetal, parameterName=ELMetal_FlimColor_Gmap_influnce}
  Multiply(A=@n31, B=Saturate(None=Add(A=@n38, B=@n39)))
  Multiply(A=@n23.RGB)
  n40 = Add(A=@n37.R) {constB=-0.5}
  n41 = Add(
   A=NamedRerouteUsage()
   B=Multiply(A=@n40, B=ScalarParameter() {defaultValue=0.2, group=EffectLiquidMetal, parameterName=ELMetal_WarpIntensity})
  ) {constB=0}
  n42 = ScalarParameter() {defaultValue=1, group=EffectLiquidMetal, parameterName=ELMetal_Contrast}
  n43 = Constant() {r=1}
  n44 = Time()
  n45 = VectorParameter() {group=EffectLiquidMetal, parameterName=ELMetal_Noise_Normals_Speed_A, defaultValue=<rgba -0.025 0.023 0 1>}
  n46 = Time()
  n47 = VectorParameter() {group=EffectLiquidMetal, parameterName=ELMetal_Noise_Normals_Speed_B, defaultValue=<rgba 0.025 -0.0133 0 1>}
  n48 = Add(
   A=MaterialFunctionCall!(
    Texture_Normals=NamedRerouteUsage()
    Rotation_Axis=NamedRerouteUsage()
    Caculated_Bounds=@n41
    Pre_Skinned_Normals=NamedRerouteUsage()
    IsStatsB=NamedRerouteUsage()
    IsState=NamedRerouteUsage()
    Blend_Contrast=@n42
    Fill_alternate_Axis_Opacity=@n43
    Scale=NamedRerouteUsage()
    Offset_Pivot
    Offset=Multiply(A=@n44, B=@n45.RGB)
   ) {materialFunction=&/MF_TopsSideProjection_Normals_V2}.Add_Surface_Normal
   B=MaterialFunctionCall(
    Texture_Normals=NamedRerouteUsage()
    Rotation_Axis=NamedRerouteUsage()
    Caculated_Bounds=@n41
    Pre_Skinned_Normals=NamedRerouteUsage()
    IsStatsB=NamedRerouteUsage()
    IsState=NamedRerouteUsage()
    Blend_Contrast=@n42
    Fill_alternate_Axis_Opacity=@n43
    Scale=NamedRerouteUsage()
    Offset=Multiply(A=@n46, B=@n47.RGB)
   ) {materialFunction=&MF_TopsSideProjection_Normals_V2}.Add_Surface_Normal
  )
  n49 = ScalarParameter() {defaultValue=0.25, group=EffectLiquidMetal, parameterName=ELMetal_Noise_Normals_Strength}
  NamedRerouteDeclaration(None=Add(A=NamedRerouteUsage(), B=Multiply(A=@n48, B=@n49))) {name=OutPutNormal, nodeColor=<rgba 0 1 0.5781 1>}
  n50 = FunctionInput(Preview=@n5.G) {bUsePreviewValueAsDefault=true, inputName=Mask Crevise, inputType=FunctionInput_Scalar, sortPriority=24, previewValue=<1 0 0 1>}
  NamedRerouteDeclaration(None=@n50) {name=Msk_Curvature, nodeColor=<rgba 0 1 0.7422 1>}
  n51 = ScalarParameter() {defaultValue=0.4, group=EffectLiquidMetal_UVs, parameterName=ELMetal_Warp_Scale_Vehicle, sortPriority=150}
  n52 = ScalarParameter() {defaultValue=0.4, group=EffectLiquidMetal_UVs, parameterName=ELMetal_Warp_Scale_Vehicle_Side, sortPriority=150}
  n53 = MaterialFunctionCall() {materialFunction=&BreakOutFloat4Components}
  n54 = LinearInterpolate(
   A=LinearInterpolate(
    A=ScalarParameter() {defaultValue=0.4, group=EffectLiquidMetal_UVs, parameterName=ELMetal_Warp_Scale_Weapon, sortPriority=150}
    B=LinearInterpolate(A=@n51, B=@n52, Alpha=@n53.G)
    Alpha=@n53.R
   )
   B=LinearInterpolate(
    A=ScalarParameter() {defaultValue=0.4, group=EffectLiquidMetal_UVs, parameterName=ELMetal_Warp_Scale_Mech, sortPriority=150}
    B=ScalarParameter() {defaultValue=0.4, group=EffectLiquidMetal_UVs, parameterName=ELMetal_Warp_Scale_Character, sortPriority=150}
    Alpha=@n53.A
   )
   Alpha=@n53.B
  )
  n55 = ScalarParameter() {defaultValue=0.4, group=EffectLiquidMetal_UVs, parameterName=ELMetal_Warp_Scale_Backpack, sortPriority=150}
  n56 = MaterialFunctionCall() {materialFunction=&BreakOutFloat2Components}
  NamedRerouteDeclaration(None=LinearInterpolate(
    A=LinearInterpolate(A=@n54, B=@n55, Alpha=@n56.R)
    B=ScalarParameter() {defaultValue=0.4, group=EffectLiquidMetal_UVs, parameterName=ELMetal_Warp_Scale_Pickaxe, sortPriority=150}
    Alpha=@n56.G
   )) {name=UVs_Warp_Scale, nodeColor=<rgba 0 0.7188 1 1>}
  NamedRerouteDeclaration(None=@n14.BaseColor) {name=Root_BaseColor, nodeColor=<rgba 0 1 0.7422 1>}
  n57 = VectorParameter() {group=EffectLiquidMetal_BC, parameterName="ELMetal_BC_Simple A=Roughness", defaultValue=<rgba 1 1 1 0.5>}
  n58 = ComponentMask(None=@n48) {g=true, r=true}
  n59 = ScalarParameter() {defaultValue=0.25, group=EffectLiquidMetal_BC, parameterName=ELMetal_BC_Warp_Noise_Normals_Strength}
  n60 = Multiply(A=@n40, B=ScalarParameter() {defaultValue=0.2, group=EffectLiquidMetal_BC, parameterName=ELMetal_BC_Warp_Intensity})
  n61 = MaterialFunctionCall!(
   Normal_Map
   Offset Rt
   Pivot
   Caculated_Bounds=NamedRerouteUsage()
   Pre_Skinned_Normals=NamedRerouteUsage()
   is_State=NamedRerouteUsage()
   UV_Rotation
   Uvs_Offset
   Uvs_TileRate=Multiply(A=@n36, B=ScalarParameter() {defaultValue=1, group=EffectLiquidMetal_BC, parameterName=ELMetal_BC_TileRate})
   Uvs_Offset_Pivot
   For_Safe_Zones
   Rotation_Comp
   Clamp Y
   Clamp X
   Clamp Padding_Comp
   No_Rotation
  ) {materialFunction=&/MF_Projection_Side_Normal_V2}
  n62 = TextureSampleParameter2D(UVs=Add(A=Multiply(A=@n58, B=@n59), B=Add(A=@n60, B=@n61.UVs))) {group=EffectLiquidMetal_BC, parameterName=ELMetal_BC_Tex, texture=/Game/Weapons/WeaponSkins/Wraps/Textures/UVCheaker}
  n63 = StaticBoolParameter() {group=EffectLiquidMetal_BC, parameterName="ELMetal__BC_Use_Tex--->Simple"}
  NamedRerouteDeclaration(None=StaticSwitch(True=@n57.A, False=@n62.A, Value=@n63)) {name=BC_Roughness, nodeColor=<rgba 1 0 0.7734 1>}
  n64 = FunctionInput(Preview=@n5.A) {bUsePreviewValueAsDefault=true, inputName=Mask Gmap, inputType=FunctionInput_Scalar, sortPriority=24, previewValue=<1 0 0 1>}
  NamedRerouteDeclaration(None=@n64) {name=Msk_Gmap, nodeColor=<rgba 0 1 0.7656 1>}
  n65 = VertexInterpolator(VS=PreSkinnedPosition())
  NamedRerouteDeclaration(None=@n65.PS) {name=PreSkin Position, nodeColor=<rgba 0 0.9297 1 1>}
  n66 = VectorParameter() {group=EffectLiquidMetal_UVs, parameterName=ELMetal_Rotation_Vehicle, defaultValue=<rgba 0 0 0 1>}
  n67 = VectorParameter() {group=EffectLiquidMetal_UVs, parameterName=ELMetal_Rotation_Vehicle_Side, defaultValue=<rgba 0 0 0 1>}
  n68 = MaterialFunctionCall(Float4=NamedRerouteUsage()) {materialFunction=&BreakOutFloat4Components}
  n69 = VectorParameter() {group=EffectLiquidMetal_UVs, parameterName=ELMetal_Rotation_Character, defaultValue=<rgba 0.25 0.25 0 1>}
  n70 = LinearInterpolate(
   A=LinearInterpolate(
    A=VectorParameter() {group=EffectLiquidMetal_UVs, parameterName=ELMetal_Rotation_Weapon, defaultValue=<rgba 0 0 -0.25 1>}.RGB
    B=LinearInterpolate(A=@n66.RGB, B=@n67.RGB, Alpha=@n68.G)
    Alpha=@n68.R
   )
   B=LinearInterpolate(
    A=VectorParameter() {group=EffectLiquidMetal_UVs, parameterName=ELMetal_Rotation_Mech, defaultValue=<rgba 0.25 0.25 0 1>}.RGB
    B=@n69.RGB
    Alpha=@n68.A
   )
   Alpha=@n68.B
  )
  n71 = VectorParameter() {group=EffectLiquidMetal_UVs, parameterName=ELMetal_Rotation_Backpack, defaultValue=<rgba 0 0 0 1>}
  n72 = MaterialFunctionCall(Float2=NamedRerouteUsage()) {materialFunction=&BreakOutFloat2Components}
  NamedRerouteDeclaration(None=LinearInterpolate(
    A=LinearInterpolate(A=@n70, B=@n71.RGB, Alpha=@n72.R)
    B=VectorParameter() {group=EffectLiquidMetal_UVs, parameterName=ELMetal_Rotation_Pickaxe, defaultValue=<rgba 0 0 0 1>}.RGB
    Alpha=@n72.G
   )) {name=TopsSide_Rotation, nodeColor=<rgba 0 1 0.9297 1>}
  n73 = ScalarParameter() {group=EffectLiquidMetal, parameterName=ELMetal_Roughness_High, sortPriority=15}
  n74 = ScalarParameter() {defaultValue=1, group=EffectLiquidMetal, parameterName=ELMetal_Roughness_Low, sortPriority=15}
  n75 = NamedRerouteUsage()
  n76 = ScalarParameter() {group=EffectLiquidMetal, parameterName=ELMetal_Roughness, sortPriority=15}
  n77 = ScalarParameter() {group=EffectLiquidMetal, parameterName="ELMetal_Roughness_Use_Root_Material   1=On", sortPriority=15}
  n78 = SetMaterialAttributes!(
   MaterialAttributes=NamedRerouteUsage()
   BaseColor=NamedRerouteUsage()
   Metallic=ScalarParameter() {defaultValue=1, group=EffectLiquidMetal, parameterName=ELMetal_Metalic, sortPriority=15}
   Roughness=StaticSwitch(True=LinearInterpolate(A=@n73, B=@n74, Alpha=@n75), False=LinearInterpolate(A=@n76, B=@n14.Roughness, Alpha=@n77), Value=NamedRerouteUsage())
   Normal=NamedRerouteUsage()
  ) {attributeSetTypes [
    #2
    #4
    #3
    #1
   ]}
  NamedRerouteDeclaration(None=@n78) {name=Mat_Base, nodeColor=<rgba 0 0.3906 1 1>}
  n79 = TextureObjectParameter() {
   group=EffectLiquidMetal
   parameterName=ELMetal_Tex__Normals
   samplerType=SAMPLERTYPE_Normal
   texture=&FlatNormalBlue
  }
  NamedRerouteDeclaration(None=@n79) {name=ELMetal_Tex__WarpTexture, nodeColor=<rgba 1 0.2109 0 1>}
  NamedRerouteDeclaration(None=StaticSwitch(True=@n57.RGB, False=@n62.RGB, Value=@n63)) {name=BaseColor, nodeColor=<rgba 1 0 0.1406 1>}
  n80 = ScalarParameter() {defaultValue=1, group=EffectLiquidMetal_UVs, parameterName=ELMetal_ScaleA_Vehicle, sortPriority=150}
  n81 = ScalarParameter() {defaultValue=1, group=EffectLiquidMetal_UVs, parameterName=ELMetal_ScaleA_Vehicle_Side, sortPriority=150}
  n82 = MaterialFunctionCall(Float4=NamedRerouteUsage()) {materialFunction=&BreakOutFloat4Components}
  n83 = LinearInterpolate(
   A=LinearInterpolate(
    A=ScalarParameter() {defaultValue=0.5, group=EffectLiquidMetal_UVs, parameterName=ELMetal_ScaleA_Weapon, sortPriority=150}
    B=LinearInterpolate(A=@n80, B=@n81, Alpha=@n82.G)
    Alpha=@n82.R
   )
   B=LinearInterpolate(
    A=ScalarParameter() {defaultValue=1, group=EffectLiquidMetal_UVs, parameterName=ELMetal_ScaleA_Mech, sortPriority=150}
    B=ScalarParameter() {defaultValue=1, group=EffectLiquidMetal_UVs, parameterName=ELMetal_ScaleA_Character, sortPriority=150}
    Alpha=@n82.A
   )
   Alpha=@n82.B
  )
  n84 = ScalarParameter() {defaultValue=1, group=EffectLiquidMetal_UVs, parameterName=ELMetal_ScaleA_Backpack, sortPriority=150}
  n85 = MaterialFunctionCall(Float2=NamedRerouteUsage()) {materialFunction=&BreakOutFloat2Components}
  NamedRerouteDeclaration(None=LinearInterpolate(
    A=LinearInterpolate(A=@n83, B=@n84, Alpha=@n85.R)
    B=ScalarParameter() {defaultValue=1, group=EffectLiquidMetal_UVs, parameterName=ELMetal_ScaleA_Pickaxe, sortPriority=150}
    Alpha=@n85.G
   )) {name=UVs_Scale_A, nodeColor=<rgba 0 0.7188 1 1>}
  n86 = ScalarParameter() {defaultValue=1, group=EffectLiquidMetal_UVs, parameterName=ELMetal_ScaleB_Vehicle, sortPriority=150}
  n87 = ScalarParameter() {defaultValue=1, group=EffectLiquidMetal_UVs, parameterName=ELMetal_ScaleB_Vehicle_Side, sortPriority=150}
  n88 = MaterialFunctionCall(Float4=NamedRerouteUsage()) {materialFunction=&BreakOutFloat4Components}
  n89 = LinearInterpolate(
   A=LinearInterpolate(
    A=ScalarParameter() {defaultValue=0.75, group=EffectLiquidMetal_UVs, parameterName=ELMetal_ScaleB_Weapon, sortPriority=150}
    B=LinearInterpolate(A=@n86, B=@n87, Alpha=@n88.G)
    Alpha=@n88.R
   )
   B=LinearInterpolate(
    A=ScalarParameter() {defaultValue=1, group=EffectLiquidMetal_UVs, parameterName=ELMetal_ScaleB_Mech, sortPriority=150}
    B=ScalarParameter() {defaultValue=1, group=EffectLiquidMetal_UVs, parameterName=ELMetal_ScaleB_Character, sortPriority=150}
    Alpha=@n88.A
   )
   Alpha=@n88.B
  )
  n90 = ScalarParameter() {defaultValue=1, group=EffectLiquidMetal_UVs, parameterName=ELMetal_ScaleB_Backpack, sortPriority=150}
  n91 = MaterialFunctionCall(Float2=NamedRerouteUsage()) {materialFunction=&BreakOutFloat2Components}
  NamedRerouteDeclaration(None=LinearInterpolate(
    A=LinearInterpolate(A=@n89, B=@n90, Alpha=@n91.R)
    B=ScalarParameter() {defaultValue=1, group=EffectLiquidMetal_UVs, parameterName=ELMetal_ScaleB_Pickaxe, sortPriority=150}
    Alpha=@n91.G
   )) {name=UVs_Scale_B, nodeColor=<rgba 0 0.7188 1 1>}
  NamedRerouteDeclaration(None=TextureObjectParameter() {group=EffectLiquidMetal, parameterName=ELMetal_Tex__Warp, texture=&T_BlacksRGB}) {name=ELMetal_Tex__Warp, nodeColor=<rgba 1 0.2109 0 1>}
  NamedRerouteDeclaration(None=FunctionInput(Preview=Constant4Vector().RGBA) {bUsePreviewValueAsDefault=true, inputName=IsState, inputType=FunctionInput_Vector4, sortPriority=4}) {name=IsState, nodeColor=<rgba 0.1497 1 0.1633 1>}
  NamedRerouteDeclaration(None=FunctionInput(Preview=Constant2Vector()) {bUsePreviewValueAsDefault=true, inputName=IsState_B, inputType=FunctionInput_Vector2, sortPriority=4}) {name=IsState_B, nodeColor=<rgba 0.1497 1 0.1633 1>}
  n92 = PreSkinnedLocalBounds()
  n93 = Subtract(A=@n65.PS, B=@n92.Min)
  n94 = MaterialFunctionCall(Color=Subtract(A=@n92.Max, B=@n92.Max)) {materialFunction=&SplitComponents}
  n95 = Max(A=Max(A=@n94.R, B=@n94.R), B=@n94.B)
  NamedRerouteDeclaration(None=FunctionInput(Preview=Divide(A=@n93, B=@n95)) {bUsePreviewValueAsDefault=true, inputName=Calcated_Bounds, sortPriority=11}) {name=Caclcated_Bounds, nodeColor=<rgba 0 0.9297 1 1>}
  NamedRerouteDeclaration(None=@n32) {name=Output_basecolor, nodeColor=<rgba 0.4766 0 1 1>}
  n96 = PreSkinnedNormal()
  NamedRerouteDeclaration(None=FunctionInput(Preview=VertexInterpolator(VS=@n96).PS) {bUsePreviewValueAsDefault=true, inputName=Pre-Skinned Local Normal, sortPriority=10}) {name=PreSkin_Normal, nodeColor=<rgba 0.8281 1 0 1>}
  NamedRerouteDeclaration(None=@n14.Normal) {name=Root_Normal, nodeColor=<rgba 0.4062 0 1 1>}
  NamedRerouteUsage()
  NamedRerouteUsage()
  NamedRerouteUsage()
 ]}
