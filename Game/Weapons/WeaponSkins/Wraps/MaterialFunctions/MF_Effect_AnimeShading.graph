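# Material function for weapon-skin wraps that applies anime toon shading to incoming material attributes, with primary, accent and crevice masks.
# Per-item-type offset, rotation and scale drive side-projected textures and normals.
# Stepped lighting and SSC thresholds choose lit or shaded colors, with optional LUT, rim, specular, hair sheen and ink lines.
# Curvature outlines are platform-switched, emissive follows time of day, and a PBR override blends with the flat toon material.
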MaterialFunction /Game/Weapons/WeaponSkins/Wraps/MaterialFunctions/MF_Effect_AnimeShading
graph {nodes [
  n1 = ScalarParameter() {defaultValue=0.85, group=Effect_Anime_TOD_Emissive, parameterName=EF_Anime_TOD__OverallNightBrighness, sliderMax=10, sortPriority=2}
  n2 = CollectionParameter() {collection=&FortniteMaterialParameters, parameterName=TimeOfDay_0-1}
  n3 = ScalarParameter() {group=00-FrontEnd-Overrides, parameterName=00-FE-TODCC-Override}
  n4 = Multiply(A=Cosine(None=Add(A=@n2, B=@n3))) {constB=0.5}
  n5 = OneMinus(None=Power(Base=Add(A=@n4) {constB=0.5}) {constExponent=8, desc=Day time EXP})
  n6 = Saturate(None=OneMinus(None=Power(Base=@n5) {desc=Night time EXP}))
  n7 = LinearInterpolate(B=@n1, Alpha=@n6) {constA=1}
  n8 = VectorParameter() {group=Effect_Anime_TOD_Emissive, parameterName=EF_Anime_TOD__Tri_Color, defaultValue=<rgba 0 0 0 1>}
  n9 = TextureSampleParameter2D(UVs=NamedRerouteUsage()) {group=Effect_Anime_TOD_Emissive, parameterName=EF_Anime_TOD__Tex, sortPriority=1, texture=&T_BlacksRGB}
  n10 = ScalarParameter() {defaultValue=2, group=Effect_Anime_TOD_Emissive, parameterName=EF_Anime_TOD__EmissiveStrength, sliderMax=10, sortPriority=2}
  n11 = Multiply(A=@n9.RGB, B=@n10)
  n12 = NamedRerouteUsage()
  n13 = DotProduct(A=CameraVectorWS(), B=PixelNormalWS())
  n14 = ScalarParameter() {defaultValue=2, group=Effect_Anime_TOD_Emissive, parameterName=EF_Anime_TOD__EmissiveFresnelExp, sliderMax=10, sortPriority=2}
  n15 = Multiply(A=@n6, B=Multiply(A=LinearInterpolate(A=@n8.RGB, B=@n11, Alpha=@n12), B=Power(Base=@n13, Exp=@n14)))
  Add(A=Multiply(B=@n7), B=@n15)
  n16 = TextureSample() {samplerType=SAMPLERTYPE_Masks, texture=/Game/Weapons/FORT_Rifles/Textures/T_Scar_CM}
  n17 = FunctionInput(Preview=@n16.R) {inputName=Mask Primary, inputType=FunctionInput_Scalar, sortPriority=24, previewValue=<1 0 0 1>}
  n18 = FunctionInput(Preview=@n16.B) {inputName=Mask Secondary, inputType=FunctionInput_Scalar, sortPriority=24, previewValue=<1 0 0 1>}
  n19 = NamedRerouteDeclaration(None=@n18) {name=Mask_Accent, nodeColor=<rgba 0 1 0.5781 1>}
  n20 = Saturate(None=Add(A=NamedRerouteDeclaration(None=@n17) {name=Mask_Primary, nodeColor=<rgba 0 1 0.5781 1>}, B=@n19))
  BlendMaterialAttributes(A=BlendMaterialAttributes(A=NamedRerouteUsage(), Alpha=@n20) {vertexAttributeBlendType=UseA}, B=NamedRerouteUsage(), Alpha=@n19) {vertexAttributeBlendType=UseA}
  n21 = MaterialFunctionCall(Float4=NamedRerouteUsage()) {materialFunction=&BreakOutFloat4Components}
  n22 = LinearInterpolate(
   A=VectorParameter() {group=Effect_Anime_UVs, parameterName=EF_Anime_Offset_Weapon_01, defaultValue=<rgba 0 0 0 1>}.RGB
   B=LinearInterpolate(
    A=VectorParameter() {group=Effect_Anime_UVs, parameterName=EF_Anime_Offset_Vehicle_01, defaultValue=<rgba 0 0 0 1>}.RGB
    B=VectorParameter() {group=Effect_Anime_UVs, parameterName=EF_Anime_Offset_VehicleSide_01, defaultValue=<rgba 0 0 0 1>}.RGB
    Alpha=@n21.G
   )
   Alpha=@n21.R
  )
  n23 = LinearInterpolate(
   A=VectorParameter() {group=Effect_Anime_UVs, parameterName=EF_Anime_Offset_Mech_01, defaultValue=<rgba 0 0 0 1>}.RGB
   B=VectorParameter() {group=Effect_Anime_UVs, parameterName=EF_Anime_Offset_Character_01, defaultValue=<rgba 0 0 0 1>}.RGB
   Alpha=@n21.A
  )
  n24 = MaterialFunctionCall(Float2=NamedRerouteUsage()) {materialFunction=&BreakOutFloat2Components}
  n25 = LinearInterpolate(
   A=LinearInterpolate(
    A=LinearInterpolate(A=@n22, B=@n23, Alpha=@n21.B)
    B=VectorParameter() {group=Effect_Anime_UVs, parameterName=EF_Anime_Offset_Backpack, defaultValue=<rgba 0 0 0 1>}.RGB
    Alpha=@n24.R
   )
   B=VectorParameter() {group=Effect_Anime_UVs, parameterName=EF_Anime_Offset_Pickaxe_01, defaultValue=<rgba 0 0 0 1>}.RGB
   Alpha=@n24.G
  )
  ComponentMask(None=@n25) {g=true, r=true}
  ComponentMask() {g=true, r=true}
  ComponentMask() {g=true, r=true}
  Constant4Vector()
  FunctionOutput(None=StaticSwitchParameter(True=NamedRerouteUsage(), False=NamedRerouteUsage()) {group=Customization, parameterName=Effect_AnimeShading}) {outputName=accent}
  n26 = Subtract(A=PreSkinnedPosition(), B=VectorParameter() {group=Toon_Shading, parameterName=NormalSphereLocation, defaultValue=<rgba 0 0 0 1>}.RGB)
  n27 = Multiply(A=@n26, B=Constant3Vector() {constant=<rgba 1 1 0 0>})
  n28 = ScalarParameter() {group=Toon_Shading, parameterName=Sphere to Cylindrical Normals, sliderMax=1}
  n29 = VertexInterpolator(VS=Normalize(VectorInput=LinearInterpolate(A=@n26, B=@n27, Alpha=@n28)))
  n30 = Multiply(A=VertexColor()) {constB=2}
  LinearInterpolate(
   A=LinearInterpolate(
    A=PixelNormalWS()
    B=Transform(None=@n29.PS) {transformSourceType=TRANSFORMSOURCE_Local}
    Alpha=ScalarParameter() {defaultValue=1, group=Toon_Shading, parameterName=EF_Anime_NormalSphereBlend, sliderMax=1}
   ) {constAlpha=0}
   B=Transform(None=Subtract(A=@n30)) {transformSourceType=TRANSFORMSOURCE_Local}
   Alpha=ScalarParameter() {group=Toon_Shading, parameterName=EF_Anime_FakeNormalBlend_Amt, sliderMax=1}
  )
  n31 = NamedRerouteUsage()
  n32 = NamedRerouteUsage()
  n33 = Multiply(A=@n31, B=@n32)
  n34 = VectorParameter() {group=Effect_Anime_Base_Color, parameterName=EF_Anime_Tri_BaseColor_Shaded, sortPriority=200, defaultValue=<rgba 0 0 0 1>}
  n35 = NamedRerouteUsage()
  n36 = TextureSampleParameter2D(UVs=@n35) {
   group=Effect_Anime_Base_Color
   parameterName=EF_Anime__Tex_BaseColor_Shaded
   sortPriority=1
   texture=&T_BlacksRGB
  }
  n37 = NamedRerouteUsage()
  n38 = TextureSampleParameter2D(UVs=@n35) {
   group=Effect_Anime_Base_Color
   parameterName=EF_Anime__Tex_BaseColor_Lite
   sortPriority=1
   texture=&T_BlacksRGB
  }
  n39 = ScalarParameter() {group=Effect_Anime_Base_Color, parameterName="EF_Anime_USE_Tex_BaseColor_Lite_Alpha Tri Msk 1=on", sortPriority=100}
  n40 = Saturate(None=Add(A=@n37, B=Multiply(A=@n38.A, B=@n39)))
  n41 = LinearInterpolate(A=@n34.RGB, B=@n36.RGB, Alpha=@n40)
  n42 = Multiply(
   A=NamedRerouteUsage()
   B=LinearInterpolate(A=ScalarParameter() {group=Effect_Anime_Base_Color, parameterName="EF_Anime_Primary_Mask  0=off 1=on"}, Alpha=NamedRerouteUsage())
  )
  n43 = ScalarParameter() {group=Effect_Anime_Base_Color, parameterName="EF_Anime_Primary_Mask Remove 1=on"}
  n44 = NamedRerouteDeclaration(None=Saturate(None=Add(A=@n42, B=@n43))) {name=Mask_Primary_State, nodeColor=<rgba 1 0 0.4219 1>}
  n45 = NamedRerouteUsage()
  n46 = Saturate(None=Add(A=@n32, B=NamedRerouteUsage()))
  n47 = Saturate(None=Multiply(A=ScalarParameter() {group=Effect_Anime_Accent, parameterName="EF_Anime_Accent_Mask  0=off 1=on", sortPriority=0}, B=NamedRerouteUsage()))
  n48 = NamedRerouteDeclaration(None=@n47) {name=Mask Accent_State, nodeColor=<rgba 0 1 0.0859 1>}
  n49 = Multiply(
   A=LinearInterpolate(A=LinearInterpolate(A=@n33, B=@n41, Alpha=@n44), B=Multiply(A=@n45, B=@n46), Alpha=@n48)
   B=ScalarParameter() {defaultValue=1, group=Effect_Anime_Base_Color, parameterName=EF_Anime_BaseColor_Shaded_Darkening}
  )
  n50 = VectorParameter() {group=Effect_Anime_Base_Color, parameterName=EF_Anime_Tri_BaseColor_Lite, sortPriority=200, defaultValue=<rgba 0 0 0 1>}
  n51 = LinearInterpolate(A=LinearInterpolate(A=@n31, B=LinearInterpolate(A=@n50.RGB, B=@n38.RGB, Alpha=@n40), Alpha=@n44), B=NamedRerouteUsage(), Alpha=@n48)
  n52 = Multiply(A=@n51, B=NamedRerouteUsage())
  n53 = NamedRerouteUsage()
  n54 = LinearInterpolate(A=NamedRerouteUsage(), B=NamedRerouteUsage(), Alpha=@n47)
  n55 = StaticBoolParameter() {group=Effect_Anime_Base_Color, parameterName=EF_Anime_UseCustome_Textures, sortPriority=0}
  n56 = StaticSwitch(True=@n51, False=@n54, Value=@n55)
  n57 = LinearInterpolate(
   A=StaticSwitch(
    True=StaticSwitchParameter(True=@n49, False=@n52) {group=Effect_Anime_Base_Color, parameterName=EF_Anime_UseShadingTexture, sortPriority=0}
    False=Multiply(A=@n54, B=@n53)
    Value=@n55
   )
   B=@n56
   Alpha=NamedRerouteUsage()
  )
  n58 = LinearInterpolate(
   A=ScalarParameter() {defaultValue=0.5, group=Effect_Anime_Base_Color, parameterName=EF_Anime_ThresholdDarkeningValue}
   Alpha=NamedRerouteUsage()
  )
  n59 = LinearInterpolate(A=LinearInterpolate(A=Multiply(A=@n57, B=@n58), B=@n57, Alpha=NamedRerouteUsage()), B=@n57, Alpha=NamedRerouteUsage())
  n60 = ScalarParameter() {defaultValue=5, group=Effect_Anime_Base_Color, parameterName=EF_Anime_RimHighlight_Strength}
  LinearInterpolate(
   A=@n59
   B=LinearInterpolate(A=@n59, B=Multiply(A=@n56, B=@n60), Alpha=NamedRerouteUsage())
   Alpha=ScalarParameter() {group=Effect_Anime_Base_Color, parameterName="EF_Anime_RimHighlight Use  0=off  1=on", sliderMax=1}
  )
  n61 = ComponentMask(None=NamedRerouteUsage()) {b=true, g=true, r=true}
  LinearInterpolate(
   A=@n61
   B=Multiply(A=MaterialFunctionCall() {materialFunction=&CameraDirectionVector}.Result) {constB=-1}
   Alpha=ScalarParameter() {group=Toon_Shading_Fake PBR, parameterName=CameraDirectionForSpecularHighlights, sliderMax=1}
  )
  n62 = TextureSampleParameter2D() {group=Toon_Shading, parameterName=InkLineColor_Texture, sortPriority=1, texture=&T_BlacksRGB}
  n63 = MaterialFunctionCall!(In=@n62.RGB, LUT, LUT Multiplier) {materialFunction=/Game/Packages/Fortress_SharedMaterials/Anime_Master_Material/MF_ApplyLUT}
  n64 = VectorParameter() {group=Toon_Shading, parameterName=InkLineColor, defaultValue=<rgba 0 0 0 1>}
  n65 = TextureSampleParameter2D() {group=Toon_Shading, parameterName=DistanceField_InkLines, sortPriority=1, texture=&T_BlacksRGB}
  n66 = TextureSampleParameter2D() {group=Toon_Shading, parameterName=DistanceField_InkLines_Mobile, sortPriority=1, texture=&T_BlacksRGB}
  n67 = StaticSwitchParameter(True=ShadingPathSwitch(Default=@n65.RGBA, Deferred=@n65.RGBA, Mobile=@n66.RGBA) {desc=Migrated from FeatureLevelSwitch}, False=@n65.RGBA) {group=Toon_Shading, parameterName=Enable Unique Mobile DF Inklines, sortPriority=0}
  n68 = Max(
   A=Reroute()
   B=Multiply(A=ChannelMaskParameter(None=@n67) {group=Toon_Shading, parameterName=DistanceField_Channel}.R) {desc=Mask Inkline Here}
  ) {desc=Inject Inklines here}
  n69 = ScalarParameter() {defaultValue=9, group=Toon_Shading, parameterName=DF_BlurPow}
  n70 = OneMinus(None=Multiply(A=Power(Base=@n68, Exp=@n69) {constExponent=3}, B=ScalarParameter() {defaultValue=1, group=Toon_Shading, parameterName=DF_BlurMult}) {constB=0.5})
  n71 = Constant() {r=1}
  n72 = ScalarParameter() {defaultValue=100, group=Effect_Anime, parameterName=DF_FarDistance, sortPriority=100}
  n73 = Subtract(A=Distance(A=WorldPosition().XYZ, B=CameraPositionWS()), B=@n72)
  n74 = Subtract(A=ScalarParameter() {defaultValue=500, group=Effect_Anime, parameterName=DF_NearDistance, sortPriority=100}, B=@n72)
  n75 = LinearInterpolate(
   A=ScalarParameter() {defaultValue=0.05, group=Toon_Shading, parameterName=DF_LineThickness_Near}
   B=ScalarParameter() {defaultValue=0.15, group=Toon_Shading, parameterName=DF_LineThickness_Far}
   Alpha=Saturate(None=Divide(A=@n73, B=@n74))
  )
  n76 = ScalarParameter() {defaultValue=0.8, group=Toon_Shading, parameterName=DF_LineHardness, sliderMax=1}
  n77 = OneMinus(None=Saturate(None=SphereMask(A=@n68, B=@n71, Radius=@n75, Hardness=@n76) {attenuationRadius=0.1}))
  n78 = Saturate(None=@n59)
  n79 = ScalarParameter() {defaultValue=0.8, group=Effect_Anime_PBR_Override, parameterName=EF_Anime_LitColor_PBR_CompensationScale, sliderMax=1, sortPriority=30}
  n80 = LinearInterpolate(
   A=MaterialFunctionCall(In=@n78) {materialFunction=&MF_ApplyLUT}.Result
   B=Multiply(A=@n79, B=MaterialFunctionCall(In=@n78) {materialFunction=&MF_ApplyLUT}.Result) {constB=0.5}
   Alpha=NamedRerouteUsage()
  )
  n81 = StaticSwitchParameter(True=@n80, False=@n59) {group=Effect_Anime, parameterName=EF-Anime-Use-Apply-LUT, sortPriority=0}
  n82 = StaticSwitchParameter(True=NamedRerouteUsage(), False=@n81) {group=Effect_Anime_Fake_PBR, parameterName=EF_Anime_Fake_PBR, sortPriority=0}
  LinearInterpolate(
   A=StaticSwitchParameter(
    True=StaticSwitchParameter(True=@n63.Result, False=@n62.RGB) {group=Toon_Shading, parameterName=Use Apply LUT, sortPriority=0}
    False=LinearInterpolate(A=@n64.RGB)
   ) {group=Toon_Shading, parameterName=UseInkLineColorTexture, sortPriority=0}
   B=@n82
   Alpha=Saturate(None=Multiply(A=@n70, B=@n77))
  )
  n83 = NamedRerouteUsage()
  n84 = Add(A=CameraVectorWS(), B=ComponentMask(None=@n83) {b=true, g=true, r=true})
  n85 = ScalarParameter() {group=Effect_Anime_Base_Color, parameterName=EF_Anime_Specular_Highlight_IgnoreZ}
  n86 = AppendVector(A=Constant2Vector() {g=1, r=1}, B=Saturate(None=OneMinus(None=@n85)))
  n87 = PixelNormalWS()
  n88 = DotProduct(A=Normalize(VectorInput=Multiply(A=@n84, B=@n86)), B=Normalize(VectorInput=Multiply(A=@n86, B=@n87)))
  n89 = MaterialFunctionCall(
   In=Power(Base=Saturate(None=@n88), Exp=ScalarParameter() {defaultValue=100, group=Effect_Anime_Base_Color, parameterName=EF_Anime_Specular_Power})
   Contrast=ScalarParameter() {defaultValue=4, group=Effect_Anime_Base_Color, parameterName=EF_Anime_Specular_Sharpness}
  ) {materialFunction=&CheapContrast}
  n90 = Divide(A=VertexInterpolator(VS=PreSkinnedPosition()).PS, B=PreSkinnedLocalBounds().Max)
  n91 = StaticSwitchParameter(
   True=TextureCoordinate() {coordinateIndex=1}
   False=StaticSwitchParameter(True=ComponentMask(None=@n90) {g=true, r=true}, False=TextureCoordinate()) {group=Toon_Shading, parameterName=UsePreskinnedPositionForSheen, sortPriority=0}
  ) {group=Toon_Shading, parameterName=UseUV2ForSheen, sortPriority=0}
  n92 = ScalarParameter() {defaultValue=7, group=Toon_Shading_Hair, parameterName=HairNoise_X_Tiling}
  n93 = ScalarParameter() {defaultValue=1, group=Toon_Shading_Hair, parameterName=HairNoise_Y_Tiling}
  n94 = TextureSample(UVs=Multiply(A=@n91, B=AppendVector(A=@n92, B=@n93))) {texture=&T_NoisePack_01}
  n95 = ScalarParameter() {defaultValue=0.1, group=Toon_Shading_Hair, parameterName=HairNoiseStrength}
  n96 = Add(
   A=MaterialFunctionCall() {materialFunction=&CameraDirectionVector}.Result
   B=ComponentMask(None=NamedRerouteUsage()) {b=true, g=true, r=true}
  )
  n97 = OneMinus(None=DotProduct(A=Normalize(VectorInput=@n96), B=Constant3Vector() {constant=<rgba 0 0 1 0>}))
  n98 = ScalarParameter() {defaultValue=1, group=Toon_Shading_Hair, parameterName=SheenMovementThreshold}
  n99 = Add(
   A=Multiply(A=@n91, B=ScalarParameter() {defaultValue=2, group=Toon_Shading_Hair, parameterName=SheenBands}) {constB=2}
   B=LinearInterpolate(
    A=ScalarParameter() {defaultValue=0.2, group=Toon_Shading_Hair, parameterName=SheenBandMinThresh}
    B=ScalarParameter() {defaultValue=0.3, group=Toon_Shading_Hair, parameterName=SheenBandMaxThresh}
    Alpha=Multiply(A=@n97, B=@n98)
   )
  )
  n100 = ChannelMaskParameter(None=Add(A=Multiply(A=@n94.G, B=@n95), B=Frac(None=@n99))) {group=Toon_Shading_Hair, parameterName=SheenDirectionChannel}
  n101 = CurveAtlasRowParameter(CurveTime=@n100.R) {
   atlas=/Game/Characters/Shared/CurveAtlas/AthenaCharacterCurveAtlas
   curve=/Game/Packages/Fortress_SharedMaterials/Anime_Master_Material/AnisoTropicHighLightCurve
   defaultValue=15
   group=Toon_Shading_Hair
   parameterName=AnisotropicSpecCurve
  }
  MaterialExpressionMaterialFunctionCall_30 = MaterialFunctionCall!(
   Index=ScalarParameter() {defaultValue=15, group=Toon_Shading_Hair, parameterName="AnisotropicSpecCurve  Index"}
   CCinput=@n101
   CCinput_Alpha
   UV=@n100.R
  ) {materialFunction=&/MF_Global_CC_Tex}
  n102 = Convert(None=@MaterialExpressionMaterialFunctionCall_30.3V) {
   nodeName=Break Float3
   convertInputs [{
     type=Vector3
     expressionInput {
      expression=&self:MaterialExpressionMaterialFunctionCall_30
      inputName=None
      mask=0
      maskA=0
      maskB=0
      maskG=0
      maskR=0
      outputIndex=0
     }
    }]
   convertOutputs [{type=Scalar}, {type=Scalar}, {type=Scalar}]
  }
  n103 = Saturate(None=@n102)
  n104 = StaticSwitchParameter(True=TextureCoordinate() {coordinateIndex=1}, False=TextureCoordinate()) {group=Toon_Shading, parameterName=UseUV2ForSheen, sortPriority=0}
  n105 = ConstantBiasScale(None=Sine(None=ComponentMask(None=@n104) {g=true}))
  n106 = Multiply(
   A=@n104
   B=AppendVector(
    A=ScalarParameter() {defaultValue=128, group=Toon_Shading_Hair, parameterName=AnisoTropix_X_Tiling}
    B=ScalarParameter() {group=Toon_Shading_Hair, parameterName=AnisoTropix_Y_Tiling}
   )
  )
  n107 = ComponentMask(None=MaterialFunctionCall(Vec2=Ceil(None=@n106)) {materialFunction=&MF_Hash}.Vec2) {r=true}
  n108 = ConstantBiasScale(None=Sine(None=Add(A=@n105, B=@n107)))
  n109 = ScalarParameter() {defaultValue=0.1, group=Toon_Shading_Hair, parameterName=HairStrandSpread}
  n110 = Saturate(None=Subtract(A=@n102, B=Multiply(A=@n108, B=@n109) {constB=0.5}))
  n111 = ScalarParameter() {defaultValue=16, group=Toon_Shading_Hair, parameterName=HairStrandPower}
  n112 = Multiply(
   A=MaterialFunctionCall(In=Power(Base=@n110, Exp=@n111), Contrast=ScalarParameter() {defaultValue=1, group=Toon_Shading_Hair, parameterName=HairStrandSharpness}) {materialFunction=&CheapContrast}.Result
   B=ScalarParameter() {defaultValue=0.5, group=Toon_Shading_Hair, parameterName=HairStrandOpacity}
  )
  n113 = NamedRerouteUsage()
  n114 = ScalarParameter() {group=Toon_Shading_Hair, parameterName=HairSpecThreshold}
  n115 = Constant() {r=1}
  LinearInterpolate(
   A=@n89.Result
   B=LinearInterpolate(
    B=Add(A=@n103, B=@n112)
    Alpha=MaterialFunctionCall(Alpha=@n113, Min=@n114, Max=@n115) {materialFunction=&SmoothStep}.Result
   )
   Alpha=ScalarParameter() {group=Toon_Shading_Hair, parameterName=UseAnisotropy, sliderMax=1}
  )
  LinearInterpolate(A=@n81, B=@n81, Alpha=NamedRerouteUsage())
  Multiply(A=@n37)
  Multiply(A=@n20)
  n116 = VectorParameter() {
   group=Effect_Anime
   parameterName=EF_Anime_Curvature
   sortPriority=51
   channelNames {a=pits power, b=Pits Strength, g=Peaks power, r="Peaks  Strength"}
   defaultValue=<rgba 8 1.3 0 1>
  }
  n117 = Saturate(None=Add(A=NamedRerouteUsage()) {constB=-0.5})
  n118 = Power(Base=OneMinus(None=Multiply(A=@n116."Peaks  Strength", B=@n117)), Exp=@n116.Peaks power)
  n119 = NamedRerouteUsage()
  n120 = Saturate(None=Multiply(A=Add(A=@n119) {constB=-0.5}) {constB=-1})
  n121 = Power(Base=OneMinus(None=Multiply(A=@n116.Pits Strength, B=@n120)), Exp=@n116.pits power)
  n122 = MaterialFunctionCall(
   In=Saturate(None=Min(A=@n118, B=@n121))
   Contrast=ScalarParameter() {defaultValue=1.5, group=Effect_Anime, parameterName=EF_Anime_Curvature_Contrast, sortPriority=51}
  ) {materialFunction=&CheapContrast}
  Multiply(A=@n122.Result)
  n123 = VectorParameter() {
   group=Effect_Anime
   parameterName=EF_Anime_Curvature near
   sortPriority=51
   channelNames {a=pits power, b=Pits Strength, g=Peaks power, r=Peaks Strength}
   defaultValue=<rgba 1.5 2 0 1>
  }
  n124 = Add(A=NamedRerouteUsage()) {constB=-0.5}
  n125 = OneMinus(None=Multiply(A=@n123.Peaks Strength, B=Saturate(None=@n124)))
  n126 = Multiply(A=Add(A=NamedRerouteUsage()) {constB=-0.5}) {constB=-1}
  n127 = OneMinus(None=Multiply(A=@n123.Pits Strength, B=Saturate(None=@n126)))
  n128 = Saturate(None=Min(A=Power(Base=@n125, Exp=@n123.Peaks power), B=Power(Base=@n127, Exp=@n123.pits power)))
  n129 = OneMinus(None=@n128)
  n130 = ScalarParameter() {defaultValue=3, group=Effect_Anime, parameterName=EF_Anime_Curvature_Contrast Near, sortPriority=51}
  Multiply(A=OneMinus(None=MaterialFunctionCall(In=@n129, Contrast=@n130) {materialFunction=&CheapContrast}.Result))
  n131 = VectorParameter() {group=Effect_Anime_SCC, parameterName=EF_Anime_SSC_Simple, defaultValue=<rgba 1 0 1 1>}
  n132 = TextureSampleParameter2D() {
   group=Effect_Anime_SCC
   parameterName=EF_Anime_SSC_Texture
   samplerType=SAMPLERTYPE_Masks
   sortPriority=1
   texture=/Game/Characters/Player/Female/Medium/Bodies/F_MED_Knight_Cat/Textures/T_F_Knight_Cat_Helmet_SSC
  }
  NamedRerouteDeclaration(None=StaticSwitchParameter(True=@n132.RGBA, False=AppendVector(A=@n131.RGB, B=@n131.RGB)) {desc=Doesn't work with Enable Far SSC, group=Effect_Anime_SCC, parameterName=EF_Anime_SSC_Use_Textures, sortPriority=0}) {name=SSC Map, nodeColor=<rgba 1 0.3516 0 1>}
  n133 = Normalize(VectorInput=PixelNormalWS())
  NamedRerouteDeclaration(None=@n133) {name=Adjusted Normals, nodeColor=<rgba 0.5 0.5 1 1>}
  n134 = Constant() {r=1}
  n135 = GetMaterialAttributes(None=FunctionInput(Preview=SetMaterialAttributes!(MaterialAttributes, BaseColor, Metallic=@n134) {attributeSetTypes [#1, #2]}) {inputName=Material, inputType=FunctionInput_MaterialAttributes}) {attributeGetTypes [
    #3
    #4
    #5
    #2
    #1
    #6
   ]}
  n136 = Multiply(
   A=@n135.Specular
   B=ScalarParameter() {defaultValue=0.5, group=Effect_Anime_Base_Color, parameterName=EF_Anime_Root_Specular_Mulplyer, sortPriority=10}
  )
  n137 = ScalarParameter() {defaultValue=0.5, group=Effect_Anime_Base_Color, parameterName=EF_Anime_Base_Specular, sortPriority=10}
  n138 = NamedRerouteUsage()
  n139 = NamedRerouteUsage()
  NamedRerouteDeclaration(None=StaticSwitch(True=LinearInterpolate(A=@n136, B=@n137, Alpha=@n138), False=@n136, Value=@n139)) {name=Root_Specular Input, nodeColor=<rgba 1 1 0 1>}
  n140 = ScalarParameter() {group=Effect_Anime_Base_Color, parameterName=EF_Anime_Base_Metalic, sortPriority=10}
  NamedRerouteDeclaration(None=StaticSwitch(True=LinearInterpolate(A=@n135.Metallic, B=@n140, Alpha=@n138), False=@n135.Metallic, Value=@n139)) {name=Root_Metallic Input, nodeColor=<rgba 1 1 0 1>}
  NamedRerouteDeclaration(None=TextureSample(Tex=TextureObjectParameter() {group=Skin FX, parameterName=SkinFX_Mask, sortPriority=0}).RGBA) {name=SkinFX, nodeColor=<rgba 1 0.3984 0 1>}
  NamedRerouteDeclaration(None=@n135.EmissiveColor) {name=Root_EmissiveColor, nodeColor=<rgba 1 1 0 1>}
  n141 = CollectionParameter() {
   collection=&FortniteMaterialParameters
   parameterName=SunlightDirectionalVector
  }
  NamedRerouteDeclaration(None=@n141) {name=LightVector, nodeColor=<rgba 0 1 0.8634 1>}
  n142 = FunctionInput(Preview=@n16.G) {inputName=Mask Crevise, inputType=FunctionInput_Scalar, sortPriority=24, previewValue=<1 0 0 1>}
  NamedRerouteDeclaration(None=@n142) {name=Crevise_map, nodeColor=<rgba 1 0.6797 0 1>}
  n143 = TextureObjectParameter() {group=Customization, parameterName="CustomizationMask (_CM)", sortPriority=-400, texture=&T_BlacksRGB}
  NamedRerouteDeclaration(None=@n143) {name=Texture_Sample_CM, nodeColor=<rgba 0 0.4375 1 1>}
  NamedRerouteDeclaration(None=@n135.MaterialAttributes) {name=Root_Material, nodeColor=<rgba 0.7812 1 0 1>}
  n144 = NamedRerouteUsage()
  n145 = ScalarParameter() {group=Effect_Anime_PBR_Override, parameterName=EF_Anime_PBR_BaseColor_EmissiveIntensity, sliderMax=1, sortPriority=19}
  n146 = NamedRerouteUsage()
  n147 = ScalarParameter() {group=Effect_Anime_PBR_Override, parameterName=EF_Anime_PBR_EmissiveIntensity, sliderMax=1, sortPriority=20}
  n148 = Constant()
  n149 = SetMaterialAttributes!(
   MaterialAttributes=NamedRerouteUsage()
   BaseColor=@n148
   Metallic=@n148
   Specular=@n148
   EmissiveColor=NamedRerouteUsage()
   Roughness=Constant() {r=1}
   AmbientOcclusion
   SubsurfaceColor
   Opacity=Constant() {r=1}
   Normal=NamedRerouteUsage()
  ) {attributeSetTypes [
    #1
    #2
    #5
    #3
    #4
    #7
    #8
    #9
    #6
   ]}
  n150 = SetMaterialAttributes!(
   MaterialAttributes=@n149
   BaseColor=@n144
   EmissiveColor=Add(A=Multiply(A=@n144, B=@n145), B=Multiply(A=@n146, B=@n147))
   AmbientOcclusion=Constant() {r=1}
   Specular=ScalarParameter() {group=Effect_Anime_PBR_Override, parameterName=EF_Anime_PBR_Specular_Override, sliderMax=1, sortPriority=21}
   Roughness=ScalarParameter() {group=Effect_Anime_PBR_Override, parameterName=EF_Anime_PBR_Roughness_Override, sliderMax=1, sortPriority=22}
   Metallic=ScalarParameter() {group=Effect_Anime_PBR_Override, parameterName=EF_Anime_PBR_Metallic_Override, sliderMax=1, sortPriority=23}
  ) {attributeSetTypes [
    #1
    #3
    #7
    #5
    #4
    #2
   ]}
  n151 = Saturate(None=ScalarParameter() {group=Effect_Anime_PBR_Override, parameterName=EF_Anime_PBR_Shading, sliderMax=1, sortPriority=10})
  NamedRerouteDeclaration(None=BlendMaterialAttributes(A=@n149, B=@n150, Alpha=NamedRerouteDeclaration(None=@n151) {name=PBR_Switch, nodeColor=<rgba 0 1 0.8566 1>})) {name=Material_Base, nodeColor=<rgba 1 0 0.8437 1>}
  NamedRerouteDeclaration(None=StaticSwitchParameter(True=@n80, False=@n78) {group=Effect_Anime, parameterName=EF-Anime-Use-Apply-LUT, sortPriority=0}) {name=Base Color OUT, nodeColor=<rgba 0 1 0.5705 1>}
  NamedRerouteDeclaration(None=FunctionInput(Preview=Constant2Vector()) {bUsePreviewValueAsDefault=true, inputName=IsState_B, inputType=FunctionInput_Vector2, sortPriority=4}) {name=IsState_B, nodeColor=<rgba 0.1497 1 0.1633 1>}
  n152 = PreSkinnedLocalBounds()
  n153 = Subtract(A=VertexInterpolator(VS=PreSkinnedPosition()).PS, B=@n152.Min)
  n154 = MaterialFunctionCall(Color=Subtract(A=@n152.Max, B=@n152.Max)) {materialFunction=&SplitComponents}
  n155 = Max(A=Max(A=@n154.R, B=@n154.R), B=@n154.B)
  NamedRerouteDeclaration(None=FunctionInput(Preview=Divide(A=@n153, B=@n155)) {bUsePreviewValueAsDefault=true, inputName=Calcated_Bounds, sortPriority=11}) {name=Caclcated_Bounds, nodeColor=<rgba 0 0.9297 1 1>}
  n156 = PreSkinnedNormal()
  NamedRerouteDeclaration(None=FunctionInput(Preview=VertexInterpolator(VS=@n156).PS) {bUsePreviewValueAsDefault=true, inputName=Pre-Skinned Local Normal, sortPriority=10}) {name=PreSkin_Normal, nodeColor=<rgba 0.8281 1 0 1>}
  NamedRerouteDeclaration(None=@n135.Normal) {name=Root_Normal, nodeColor=<rgba 1 1 0 1>}
  n157 = TextureObjectParameter() {
   group=Effect_Anime_Base_Color
   parameterName=EF_Anime__Tex_Normals
   samplerType=SAMPLERTYPE_Normal
   sortPriority=9
   texture=&FlatNormalBlue
  }
  n158 = MaterialFunctionCall(Float4=NamedRerouteUsage()) {materialFunction=&BreakOutFloat4Components}
  n159 = MaterialFunctionCall!(
   For_Safe_Zones
   Normal_Map=@n157
   Caculated_Bounds=NamedRerouteUsage()
   Pre_Skinned_Normals=NamedRerouteUsage()
   Is Vehicle=@n158.R
   Is VehicleSide=@n158.R
   Is Mech=@n158.R
   Vehicle_Rotation
   UV_Rotation Bypass=NamedRerouteUsage()
   Mech_Rotation
   Uvs_Offset=NamedRerouteUsage()
   Offset-after rotation
   Uvs_TileRate=NamedRerouteUsage()
  ) {materialFunction=&/MF_Projection_Side_Normal}
  NamedRerouteDeclaration(None=@n159.UVs) {name=uvs_sideprojection, nodeColor=<rgba 0.5703 1 0 1>}
  n160 = DotProduct(A=ComponentMask(None=NamedRerouteUsage()) {b=true, g=true, r=true}, B=PixelNormalWS())
  n161 = OneMinus(None=Fresnel(
    ExponentIn=ScalarParameter() {defaultValue=4, group=Effect_Anime_Lighting, parameterName=EF_Anime_BL_RimHighlightExponent1}
    Normal=PixelNormalWS()
   ) {baseReflectFraction=0})
  n162 = ScalarParameter() {defaultValue=4, group=Effect_Anime_Lighting, parameterName=EF_Anime_BL_Rim_OutlineExponent2}
  n163 = Saturate(None=LinearInterpolate(
    A=ScalarParameter() {defaultValue=-4, group=Effect_Anime_Lighting, parameterName=EF_Anime_BL_Rim_OutlineC1}
    B=ScalarParameter() {defaultValue=12, group=Effect_Anime_Lighting, parameterName=EF_Anime_BL_Rim_OutlineC2}
    Alpha=Power(Base=@n161, Exp=@n162)
   ))
  NamedRerouteDeclaration(None=Multiply(A=Saturate(None=@n160), B=OneMinus(None=@n163))) {name=Rim Highlight, nodeColor=<rgba 1 0 0.1172 1>}
  n164 = VectorParameter() {group=Effect_Anime_Accent, parameterName=EF_Anime_Accent_BaseColor, sortPriority=10, defaultValue=<rgba 0.0403 0.0387 1 1>}
  n165 = VectorParameter() {group=Effect_Anime_Accent, parameterName=EF_Anime_Accent_BaseColor_Shaded, sortPriority=10, defaultValue=<rgba 0.0403 0.0387 1 1>}
  n166 = ScalarParameter() {group=Effect_Anime_Accent, parameterName="EF_Anime_Accent_USE_BaseColor_Shaded   1=on", sortPriority=0}
  n167 = Multiply(A=OneMinus(None=NamedRerouteUsage()), B=@n166)
  n168 = StaticSwitchParameter(True=LinearInterpolate(A=@n164.RGB, B=@n165.RGB, Alpha=Saturate(None=@n167)), False=@n164.RGB) {group=Effect_Anime_Base_Color, parameterName=EF_Anime_UseShadingTexture, sortPriority=0}
  n169 = NamedRerouteUsage()
  n170 = SetMaterialAttributes!(
   MaterialAttributes=NamedRerouteUsage()
   BaseColor=NamedRerouteDeclaration(None=StaticSwitch(True=@n168, False=@n164.RGB, Value=@n169)) {name=Accent_Base_Color, nodeColor=<rgba 1 0 0.6328 1>}
   Roughness=NamedRerouteDeclaration(None=ScalarParameter() {defaultValue=0.2, group=Effect_Anime_Accent, parameterName=EF_Anime_Accent_Roughness, sortPriority=10}) {name=Accent_Rougness, nodeColor=<rgba 1 0 0.6328 1>}
   Specular=NamedRerouteDeclaration(None=ScalarParameter() {defaultValue=0.5, group=Effect_Anime_Accent, parameterName=EF_Anime_Accent_Specular, sortPriority=10}) {name=Accent_Spec, nodeColor=<rgba 1 0 0.6328 1>}
   EmissiveColor
   Metallic=NamedRerouteDeclaration(None=ScalarParameter() {defaultValue=1, group=Effect_Anime_Accent, parameterName=EF_Anime_Accent_Metalic, sortPriority=10}) {name=Accent_Metalic, nodeColor=<rgba 1 0 0.6328 1>}
   Normal
  ) {attributeSetTypes [
    #1
    #4
    #5
    #3
    #2
    #6
   ]}
  NamedRerouteDeclaration(None=@n170) {name=Accent_Material, nodeColor=<rgba 1 0 0.6328 1>}
  n171 = ScalarParameter() {defaultValue=1, group=Effect_Anime_UVs, parameterName=EF_Anime_Scale_Vehicle, sortPriority=150}
  n172 = ScalarParameter() {defaultValue=1, group=Effect_Anime_UVs, parameterName=EF_Anime_Scale_Vehicle_Side, sortPriority=150}
  n173 = MaterialFunctionCall(Float4=NamedRerouteUsage()) {materialFunction=&BreakOutFloat4Components}
  n174 = LinearInterpolate(
   A=LinearInterpolate(
    A=ScalarParameter() {defaultValue=1, group=Effect_Anime_UVs, parameterName=EF_Anime_Scale_Weapon, sortPriority=150}
    B=LinearInterpolate(A=@n171, B=@n172, Alpha=@n173.G)
    Alpha=@n173.R
   )
   B=LinearInterpolate(
    A=ScalarParameter() {defaultValue=1, group=Effect_Anime_UVs, parameterName=EF_Anime_Scale_Mech, sortPriority=150}
    B=ScalarParameter() {defaultValue=1, group=Effect_Anime_UVs, parameterName=EF_Anime_Scale_Character, sortPriority=150}
    Alpha=@n173.A
   )
   Alpha=@n173.B
  )
  n175 = ScalarParameter() {defaultValue=1, group=Effect_Anime_UVs, parameterName=EF_Anime_Scale_Backpack, sortPriority=150}
  n176 = MaterialFunctionCall(Float2=NamedRerouteUsage()) {materialFunction=&BreakOutFloat2Components}
  NamedRerouteDeclaration(None=LinearInterpolate(
    A=LinearInterpolate(A=@n174, B=@n175, Alpha=@n176.R)
    B=ScalarParameter() {defaultValue=1, group=Effect_Anime_UVs, parameterName=EF_Anime_Scale_Pickaxe, sortPriority=150}
    Alpha=@n176.G
   )) {name=Side_Scale, nodeColor=<rgba 0 0.7188 1 1>}
  n177 = LinearInterpolate(
   A=ScalarParameter() {defaultValue=1, group=Effect_Anime, parameterName=EF_Anime_Root_Brightness}
   B=ScalarParameter() {defaultValue=1, group=Effect_Anime, parameterName=EF_Anime_BC_Brightness}
   Alpha=@n20
  )
  NamedRerouteDeclaration(None=Multiply(A=@n135.BaseColor, B=@n177)) {name=Root_Base_Color Input, nodeColor=<rgba 1 1 0 1>}
  n178 = VectorParameter() {group=Effect_Anime_Base_Color, parameterName=EF_Anime_BaseColor_Tint, defaultValue=<rgba 0.054 0.054 0.054 1>}
  NamedRerouteDeclaration(None=Multiply(A=@n178.RGB, B=ScalarParameter() {defaultValue=1, group=Effect_Anime_Base_Color, parameterName=EF_Anime_BaseColor_Tint_Darkening})) {name=Toon_BC_Tint, nodeColor=<rgba 0 1 0.9063 1>}
  n179 = ScalarParameter() {group=Effect_Anime_UVs, parameterName=EF_Anime_Rotation_Vehicle, sortPriority=150}
  n180 = ScalarParameter() {group=Effect_Anime_UVs, parameterName=EF_Anime_Rotation_Vehicle_Side, sortPriority=150}
  n181 = MaterialFunctionCall(Float4=NamedRerouteUsage()) {materialFunction=&BreakOutFloat4Components}
  n182 = LinearInterpolate(
   A=LinearInterpolate(
    A=ScalarParameter() {group=Effect_Anime_UVs, parameterName=EF_Anime_Rotation_Weapon, sortPriority=150}
    B=LinearInterpolate(A=@n179, B=@n180, Alpha=@n181.G)
    Alpha=@n181.R
   )
   B=LinearInterpolate(
    A=ScalarParameter() {group=Effect_Anime_UVs, parameterName=EF_Anime_Rotation_Mech, sortPriority=150}
    B=ScalarParameter() {group=Effect_Anime_UVs, parameterName=EF_Anime_Rotation_Character, sortPriority=150}
    Alpha=@n181.A
   )
   Alpha=@n181.B
  )
  n183 = ScalarParameter() {group=Effect_Anime_UVs, parameterName=EF_Anime_Rotation_Backpack, sortPriority=150}
  n184 = MaterialFunctionCall(Float2=NamedRerouteUsage()) {materialFunction=&BreakOutFloat2Components}
  NamedRerouteDeclaration(None=LinearInterpolate(
    A=LinearInterpolate(A=@n182, B=@n183, Alpha=@n184.R)
    B=ScalarParameter() {group=Effect_Anime_UVs, parameterName=EF_Anime_Rotation_Pickaxe, sortPriority=150}
    Alpha=@n184.G
   )) {name=Side_Rotation, nodeColor=<rgba 0 0.7188 1 1>}
  NamedRerouteDeclaration(None=@n25) {name=Side_offset, nodeColor=<rgba 0 0.625 1 1>}
  n185 = NamedRerouteUsage()
  n186 = LinearInterpolate(A=Add(A=@n159.Raw_Normal, B=@n185), B=@n185, Alpha=NamedRerouteUsage())
  NamedRerouteDeclaration(None=StaticSwitch(
    True=StaticSwitchParameter(True=@n186, False=@n185) {group=Effect_Anime_Base_Color, parameterName=EF_Anime_Custom_Normals, sortPriority=0}
    False=@n185
    Value=NamedRerouteUsage()
   )) {name=Material_Normal, nodeColor=<rgba 1 0 0.0937 1>}
  n187 = FunctionInput(Preview=Constant4Vector().RGBA) {bUsePreviewValueAsDefault=true, inputName=IsState, inputType=FunctionInput_Vector4, sortPriority=4}
  NamedRerouteDeclaration(None=MaterialFunctionCall(Float4=NamedRerouteDeclaration(None=@n187) {name=IsState, nodeColor=<rgba 0.1497 1 0.1633 1>}) {materialFunction=&BreakOutFloat4Components}.A) {name=IsCharacter, nodeColor=<rgba 1 0 0.7734 1>}
  n188 = NamedRerouteUsage()
  n189 = ChannelMaskParameter(None=@n188) {group=Effect_Anime_SCC, parameterName=EF_Anime_SSC_ThresholdChannel}
  n190 = ScalarParameter() {defaultValue=1, group=Effect_Anime_Accent, parameterName=EF_Anime_Accent_SSC_Threshold, sortPriority=10}
  n191 = NamedRerouteUsage()
  n192 = MaterialFunctionCall(
   Alpha=Power(
    Base=LinearInterpolate(A=@n189.R, B=@n190, Alpha=@n191)
    Exp=ScalarParameter() {defaultValue=2, group=Effect_Anime_Lighting, parameterName=EF_Anime_ThresholdmaskPower}
   )
   Min=ScalarParameter() {group=Effect_Anime_Lighting, parameterName=EF_Anime_Shaded_Threshold_Min}
   Max=ScalarParameter() {defaultValue=1, group=Effect_Anime_Lighting, parameterName=EF_Anime_Shaded_Threshold_Max}
  ) {materialFunction=&SmoothStep}
  NamedRerouteDeclaration(None=Round(None=@n192.Result)) {name=SSC Throshold, nodeColor=<rgba 0.7973 1 0 1>}
  NamedRerouteDeclaration(None=@n55) {name=Custom_texture_State, nodeColor=<rgba 1 0 0.2812 1>}
  NamedRerouteDeclaration(None=@n36.A) {name=BC_Roughness_Shaded, nodeColor=<rgba 0.7578 1 0 1>}
  n193 = TextureSampleParameter2D(UVs=@n35) {group=Effect_Anime_Base_Color, parameterName=EF_Anime__Tex_MSK, sortPriority=1, texture=&T_BlacksRGB}
  NamedRerouteDeclaration(None=@n193.RGBA) {name=BC_MSK, nodeColor=<rgba 0 1 0.7188 1>}
  n194 = Cosine(None=Add(
    A=CollectionParameter() {collection=&FortniteMaterialParameters, parameterName=TimeOfDay_0-1}
    B=ScalarParameter() {group=00-FrontEnd-Overrides, parameterName=00-FE-TODCC-Override}
   ))
  n195 = Power(Base=Add(A=Multiply(A=@n194) {constB=0.5}) {constB=0.5}) {constExponent=4}
  n196 = OneMinus(None=Power(Base=OneMinus(None=@n195)) {constExponent=4})
  n197 = TextureSampleParameter2D(UVs=NamedRerouteUsage()) {group=Effect_Anime_TOD_Emissive, parameterName=EF_Anime_TOD__Tex, sortPriority=1, texture=&T_BlacksRGB}
  n198 = ScalarParameter() {defaultValue=2, group=Effect_Anime_TOD_Emissive, parameterName=EF_Anime_TOD__EmissiveStrength, sliderMax=10, sortPriority=2}
  n199 = Multiply(A=@n197.RGB, B=@n198)
  NamedRerouteDeclaration(None=LinearInterpolate(
    A=VectorParameter() {group=Effect_Anime_TOD_Emissive, parameterName=EF_Anime_TOD__Tri_Color, defaultValue=<rgba 0 0 0 1>}.RGB
    B=Multiply(A=@n196, B=@n199)
    Alpha=NamedRerouteUsage()
   )) {name=TOD_Emissive, nodeColor=<rgba 0 0.8359 1 1>}
  n200 = MaterialFunctionCall(Color=Abs(None=NamedRerouteUsage())) {materialFunction=&SplitComponents}
  n201 = MaterialFunctionCall(Float4=NamedRerouteUsage()) {materialFunction=&BreakOutFloat4Components}
  n202 = LinearInterpolate(A=@n200.B, B=@n200.B, Alpha=@n201.G)
  n203 = MaterialFunctionCall(
   In=LinearInterpolate(A=LinearInterpolate(A=@n200.R, B=@n202, Alpha=@n201.R), B=@n200.G, Alpha=@n201.B)
   Contrast=ScalarParameter() {defaultValue=1, group=Effect_Anime_Base_Color, parameterName=EF_Anime_Tri_Blend_Contrast, sortPriority=200}
  ) {materialFunction=&CheapContrast}
  n204 = LinearInterpolate(
   A=ScalarParameter() {defaultValue=1, group=Effect_Anime_Base_Color, parameterName="EF_Anime_Tri_Blend_Blend 0=on", sortPriority=200}
   B=ScalarParameter() {defaultValue=1, group=Effect_Anime_Base_Color, parameterName="EF_Anime_Tri_Blend_Blend Character 0=on", sortPriority=200}
   Alpha=@n201.A
  )
  NamedRerouteDeclaration(None=Saturate(None=Add(A=@n203.Result, B=@n204))) {name=Tri_Mask, nodeColor=<rgba 1 0.4688 0 1>}
  n205 = Constant()
  NamedRerouteDeclaration(None=StaticSwitch(
    True=StaticSwitchParameter(True=@n166, False=@n205) {group=Effect_Anime_Base_Color, parameterName=EF_Anime_UseShadingTexture, sortPriority=0}
    False=@n205
    Value=@n169
   )) {name=Accent_BaseColor_Shaded_State, nodeColor=<rgba 1 0 0.9375 1>}
  NamedRerouteDeclaration(None=@n15) {name=TOD_Emissive_Add, nodeColor=<rgba 0 0.8359 1 1>}
  NamedRerouteDeclaration(None=@n7) {name=TOD_Emissive_Multplyer, nodeColor=<rgba 0 0.8359 1 1>}
  n206 = ComponentMask(None=NamedRerouteUsage()) {b=true, g=true, r=true}
  n207 = DotProduct(A=@n206, B=NamedRerouteUsage())
  n208 = Clamp(None=@n207)
  NamedRerouteDeclaration(None=Round(None=@n208)) {name=Shading Dot, nodeColor=<rgba 0.7973 1 0 1>}
  NamedRerouteDeclaration(None=@n38.A) {name=BC_Roughness, nodeColor=<rgba 0.7578 1 0 1>}
  n209 = ScalarParameter() {defaultValue=1.5, group=Effect_Anime, parameterName=EF_Anime_GlobalBrightness}
  n210 = ScalarParameter() {group=Effect_Anime, parameterName="EF_Anime-Outline-Disable 0=off 1=on Near", sortPriority=49}
  n211 = ScalarParameter() {group=Effect_Anime, parameterName="EF_Anime-Outline-Disable 0=off 1=on -------------------------------------", sortPriority=49}
  n212 = Distance(A=WorldPosition().XYZ, B=CameraPositionWS())
  n213 = ScalarParameter() {defaultValue=25, group=Effect_Anime, parameterName=EF_Anime_FarDistance, sortPriority=100}
  n214 = ScalarParameter() {defaultValue=100, group=Effect_Anime, parameterName=EF_Anime_NearDistance, sortPriority=100}
  n215 = Saturate(None=Divide(A=Subtract(A=@n212, B=@n213), B=Subtract(A=@n214, B=@n213)))
  n216 = LinearInterpolate(
   A=MaterialFunctionCall(In=@n128, Contrast=@n130) {materialFunction=&CheapContrast}.Result
   B=@n122.Result
   Alpha=@n215
  )
  n217 = LinearInterpolate(
   A=ScalarParameter() {defaultValue=10, group=Effect_Anime, parameterName=EF_Anime_Curvature_Mask A, sortPriority=51}
   B=ScalarParameter() {defaultValue=-7, group=Effect_Anime, parameterName=EF_Anime_Curvature_Mask B, sortPriority=51}
   Alpha=NamedRerouteUsage()
  )
  n218 = Add(
   A=StaticSwitchParameter(True=LinearInterpolate(A=@n210, B=@n211, Alpha=@n215), False=@n211) {group=Effect_Anime, parameterName=EF_Anime-Outline-Disable CameraBased, sortPriority=48}
   B=Saturate(None=StaticSwitchParameter(True=@n216, False=@n217) {group=Effect_Anime, parameterName=EF_Anime_Curvature_Mask CameraBased, sortPriority=50})
  )
  n219 = Constant() {r=1}
  n220 = MaterialFunctionCall("Xb1,PS4"=@n218, "PC-High&&Epic,XSX,PS5"=@n218, "PC-Perf,Android"=@n219, Switch=@n219, "PC-Low(NotPerf)&&Medium"=@n219) {materialFunction=&PlatformSwitch_1D}
  n221 = NamedRerouteDeclaration(None=Multiply(A=Multiply(A=@n82, B=@n209), B=Saturate(None=@n220.Result))) {name=Final Color, nodeColor=<rgba 0.875 1 0 1>}
  n222 = ScalarParameter() {defaultValue=70, group=Effect_Anime_Lighting, parameterName="EF_Anime_TOD_BrightnessCurve  Index"}
  n223 = Add(
   A=CollectionParameter() {collection=&FortniteMaterialParameters, parameterName=TimeOfDay_0-1}
   B=ScalarParameter() {group=00-FrontEnd-Overrides, parameterName=00-FE-TODCC-Override}
  )
  n224 = CurveAtlasRowParameter(CurveTime=@n223) {
   atlas=/Game/Weapons/WeaponSkins/Wraps/Textures/CruveAtlas_Wraps_01
   curve=/Game/Packages/Fortress_SharedMaterials/Anime_Master_Material/ToonTODBrightness_Curve
   defaultValue=70
   group=Effect_Anime_Lighting
   parameterName=EF_Anime_TOD_BrightnessCurve
  }
  n225 = VectorParameter() {group=Effect_Anime_Lighting, parameterName=EF_Anime_TOD_Mornning, defaultValue=<rgba 0.5 0.4992 0.45 1>}
  n226 = StaticSwitchParameter(
   True=StaticComponentMaskParameter(None=MaterialFunctionCall(Index=@n222, CCinput=@n224, CCinput_Alpha=@n224, UV=@n223) {materialFunction=&MF_Global_CC_Tex}.4v) {defaultB=true, group=Effect_Anime_Lighting, parameterName=EF_Anime_TOD_CurveChannelMask}
   False=MaterialFunctionCall!(
    Color_Morning=@n225.RGB
    Color_Day=VectorParameter() {group=Effect_Anime_Lighting, parameterName=EF_Anime_TOD_Day, defaultValue=<rgba 1 1 1 1>}.RGB
    Color_Evening=VectorParameter() {group=Effect_Anime_Lighting, parameterName=EF_Anime_TOD_Evvning, defaultValue=<rgba 0.5 0.4624 0.45 1>}.RGB
    Color_Night=VectorParameter() {group=Effect_Anime_Lighting, parameterName=EF_Anime_TOD_Night, defaultValue=<rgba 0.4262 0.425 0.5 1>}.RGB
   ) {materialFunction=&/Wrap_Color_Time_of_Day}.Result
  ) {group=Effect_Anime_Lighting, parameterName="EF_Anime_TOD_Color---->CC"}
  NamedRerouteDeclaration(None=Add(A=NamedRerouteUsage(), B=Multiply(A=@n221, B=@n226))) {name=TodState, nodeColor=<rgba 0.5 0 1 1>}
  NamedRerouteDeclaration(None=@n135.Metallic) {name=Root_Metallic, nodeColor=<rgba 1 1 0 1>}
  n227 = NamedRerouteUsage()
  n228 = Multiply(A=@n227, B=NamedRerouteUsage())
  n229 = Add(
   A=NamedRerouteUsage()
   B=ScalarParameter() {group=Effect_Anime_TOD_Emissive, parameterName="EF_Anime_TOD__Disable Mask 0=off", sliderMax=10, sortPriority=2}
  )
  n230 = Multiply(
   A=NamedRerouteUsage()
   B=ScalarParameter() {group=Effect_Anime_TOD_Emissive, parameterName="EF_Anime_TOD__Accent Mask  0=off", sliderMax=10, sortPriority=2}
  )
  n231 = Multiply(A=NamedRerouteUsage(), B=Saturate(None=Add(A=@n229, B=@n230)))
  NamedRerouteDeclaration(None=StaticSwitchParameter(True=Add(A=@n228, B=@n231), False=@n227) {group=Effect_Anime_TOD_Emissive, parameterName=EF_Anime_TOD_Emissive, sortPriority=0}) {name=TOD_Result, nodeColor=<rgba 0.2422 0 1 1>}
  n232 = ScalarParameter() {group=Effect_Anime_Lighting, parameterName=EF_Anime_Lit Bias}
  n233 = ScalarParameter() {defaultValue=-4, group=Effect_Anime_Lighting, parameterName=EF_Anime_BackLitRim_OutlineC1}
  n234 = ScalarParameter() {defaultValue=12, group=Effect_Anime_Lighting, parameterName=EF_Anime_BackLitRim_OutlineC2}
  n235 = ScalarParameter() {defaultValue=1, group=Effect_Anime_Lighting, parameterName=EF_Anime_BackLitExponent1}
  n236 = NamedRerouteUsage()
  n237 = Power(
   Base=OneMinus(None=Fresnel(ExponentIn=@n235, Normal=@n236) {baseReflectFraction=0})
   Exp=ScalarParameter() {defaultValue=4, group=Effect_Anime_Lighting, parameterName=EF_Anime_BacklitOutlineExponent2}
  )
  n238 = OneMinus(None=Saturate(None=LinearInterpolate(A=@n233, B=@n234, Alpha=@n237)))
  n239 = DotProduct(A=MaterialFunctionCall() {materialFunction=&CameraDirectionVector}.Result, B=@n206)
  n240 = ScalarParameter() {defaultValue=0.25, group=Effect_Anime_Lighting, parameterName=EF_Anime_BackLitBias}
  n241 = LinearInterpolate(
   A=Multiply(A=Add(A=@n207, B=@n232), B=@n192.Result)
   B=Multiply(A=Add(A=@n208, B=@n238), B=@n192.Result)
   Alpha=Clamp(
    None=Add(A=@n239, B=@n240)
    Max=ScalarParameter() {defaultValue=0.75, group=Effect_Anime_Lighting, parameterName=EF_Anime_BackLitMaxBias, sliderMax=1}
   )
  )
  n242 = CurveAtlasRowParameter(CurveTime=@n241) {
   atlas=&CruveAtlas_Wraps_01
   curve=/Game/Packages/Fortress_SharedMaterials/Anime_Master_Material/TwoToneShadingCurve
   defaultValue=71
   group=Effect_Anime_Lighting
   parameterName=EF_Anime_Shading_Curve
  }
  n243 = MaterialFunctionCall(
   Index=ScalarParameter() {group=Effect_Anime_Lighting, parameterName="EF_Anime_Shading_Curve  index", sortPriority=71}
   CCinput=@n242
   CCinput_Alpha=@n242
   UV=@n241
  ) {materialFunction=&MF_Global_CC_Tex}
  NamedRerouteDeclaration(None=LinearInterpolate(A=ChannelMaskParameter(None=@n243.4v) {group=Effect_Anime_Lighting, parameterName=EF_Anime_ShadingCurveChannel}.R, Alpha=NamedRerouteUsage())) {name=Shading Curve, nodeColor=<rgba 0.7973 1 0 1>}
  n244 = VectorParameter() {group=Effect_Anime_Fake_PBR, parameterName=EF_Anime_Specular Non Masked Tint, defaultValue=<rgba 1 1 1 1>}
  n245 = NamedRerouteUsage()
  n246 = VectorParameter() {group=Effect_Anime_Fake_PBR, parameterName=EF_Anime_Specular Tint, defaultValue=<rgba 1 1 1 1>}
  n247 = VectorParameter() {group=Effect_Anime_Fake_PBR, parameterName=EF_Anime_Specular Tint Metalic, sortPriority=100, defaultValue=<rgba 1 1 1 1>}
  n248 = Saturate(None=Multiply(
    A=NamedRerouteUsage()
    B=ScalarParameter() {group=Effect_Anime_Fake_PBR, parameterName="EF_Anime_Specular Strength Metalic use 1=active", sortPriority=99}
   ))
  n249 = NamedRerouteUsage()
  n250 = Multiply(A=LinearInterpolate(A=@n246.RGB, B=@n247.RGB, Alpha=@n248), B=@n249) {constA=1}
  n251 = NamedRerouteUsage()
  n252 = VectorParameter() {group=Effect_Anime_Accent, parameterName=EF_Anime_Accent_Specular Tint, defaultValue=<rgba 1 1 1 1>}
  n253 = Multiply(A=@n249, B=@n252.RGB)
  n254 = NamedRerouteUsage()
  n255 = ScalarParameter() {defaultValue=1, group=Effect_Anime_Fake_PBR, parameterName="EF_Anime_Specular Non Masked  0=On"}
  n256 = Add(A=NamedRerouteUsage(), B=NamedRerouteUsage())
  n257 = LinearInterpolate(
   A=LinearInterpolate(A=@n244.RGB, B=Multiply(A=@n244.RGB, B=@n245), Alpha=NamedRerouteUsage())
   B=LinearInterpolate(A=LinearInterpolate(A=@n246.RGB, B=@n250, Alpha=@n251), B=LinearInterpolate(A=@n252.RGB, B=@n253, Alpha=@n254), Alpha=NamedRerouteUsage())
   Alpha=Saturate(None=Add(A=@n255, B=@n256))
  )
  n258 = NamedRerouteUsage()
  n259 = Multiply(
   A=@n257
   B=Multiply(
    A=ScalarParameter() {defaultValue=1, group=Effect_Anime_Fake_PBR, parameterName=EF_Anime_Specular Strength}
    B=LinearInterpolate(Alpha=@n258) {constA=1, constB=0.25}
   )
  )
  n260 = ScalarParameter() {defaultValue=0.25, group=Effect_Anime_PBR_Override, parameterName=EF_Anime_PBR_Metallic_Min_Spec_Strenght, sliderMax=1, sortPriority=23}
  n261 = Multiply(
   A=Multiply(
    A=ScalarParameter() {defaultValue=1, group=Effect_Anime_Fake_PBR, parameterName=EF_Anime_Specular Strength Metalic, sortPriority=100}
    B=LinearInterpolate(B=@n260, Alpha=@n258) {constA=1, constB=0.25}
   )
   B=@n257
  )
  n262 = NamedRerouteUsage()
  n263 = LinearInterpolate(
   A=ScalarParameter() {defaultValue=0.0005, group=Effect_Anime_Fake_PBR, parameterName=EF_Anime_Roughness Min, sliderMax=1}
   B=ScalarParameter() {defaultValue=1, group=Effect_Anime_Fake_PBR, parameterName=EF_Anime_Roughness Max, sliderMax=1}
   Alpha=Saturate(None=Power(Base=@n262) {constExponent=3})
  )
  n264 = NamedRerouteUsage()
  n265 = NamedRerouteUsage()
  n266 = CameraVectorWS()
  n267 = ScalarParameter() {group=Effect_Anime_Fake_PBR, parameterName=EF_Anime_Ignore Z, sliderMax=1}
  n268 = AppendVector(A=Constant2Vector() {g=1, r=1}, B=Saturate(None=OneMinus(None=@n267)))
  n269 = Normalize(VectorInput=Multiply(A=Add(A=@n266, B=@n61), B=@n268))
  n270 = Normalize(VectorInput=Multiply(A=@n268, B=NamedRerouteUsage()))
  n271 = SmoothStep(Min=OneMinus(None=LinearInterpolate(A=@n263, B=@n264, Alpha=@n265)), Max=Constant() {r=1}, Value=Saturate(None=DotProduct(A=@n269, B=@n270)))
  n272 = MaterialFunctionCall(In=@n271, Contrast=ScalarParameter() {defaultValue=2, group=Effect_Anime_Fake_PBR, parameterName=EF_Anime_Specular Sharpness}) {materialFunction=&CheapContrast}
  n273 = Round(None=@n271)
  n274 = ScalarParameter() {group=Effect_Anime_Fake_PBR, parameterName=EF_Anime_Bool Specular, sliderMax=1}
  n275 = Multiply(
   A=Multiply(A=LinearInterpolate(A=@n259, B=@n261, Alpha=@n248), B=LinearInterpolate(A=@n272.Result, B=@n273, Alpha=@n274))
   B=LinearInterpolate(A=NamedRerouteUsage(), B=NamedRerouteUsage(), Alpha=NamedRerouteUsage())
  ) {constA=1}
  n276 = NamedRerouteUsage()
  NamedRerouteDeclaration(None=Add(A=Multiply(A=@n275, B=@n276), B=NamedRerouteUsage())) {name=Fake PBR, nodeColor=<rgba 0 0.2266 1 1>}
  n277 = ChannelMaskParameter(None=@n188) {group=Effect_Anime_SCC, maskChannel=Blue, parameterName=EF_Anime_SSC_SheenMaskChannel, defaultValue=<rgba 0 0 1 0>}
  NamedRerouteDeclaration(None=LinearInterpolate(A=@n277.R, Alpha=NamedRerouteUsage())) {name=SSC_Blue, nodeColor=<rgba 0 0.1118 1 1>}
  n278 = LinearInterpolate(
   A=ScalarParameter() {defaultValue=0.2, group=Effect_Anime_Base_Color, parameterName=EF_Anime_Base_Roughness, sortPriority=10}
   B=StaticSwitchParameter(True=NamedRerouteUsage(), False=NamedRerouteUsage()) {group=Effect_Anime_Base_Color, parameterName=EF_Anime_UseShadingTexture, sortPriority=0}
   Alpha=StaticSwitchParameter(
    True=ScalarParameter() {group=Effect_Anime_Base_Color, parameterName="EF_Anime_Base_Roughness_Use_Tex_Shade  1=on", sortPriority=10}
    False=Constant()
   ) {group=Effect_Anime_Base_Color, parameterName=EF_Anime_UseShadingTexture, sortPriority=0}
  )
  NamedRerouteDeclaration(None=StaticSwitch(True=LinearInterpolate(A=@n135.Roughness, B=@n278, Alpha=@n138), False=@n135.Roughness, Value=@n139)) {name=Root_Roughness Input, nodeColor=<rgba 1 1 0 1>}
  OneMinus(None=@n216)
  QualitySwitch(Default=@n218, Low=@n219, High=@n218, Medium=@n218, Epic=@n218)
  Saturate(None=Multiply(A=@n89.Result, B=ScalarParameter() {group=Effect_Anime_Base_Color, parameterName="EF_Anime_Specular_Fake  0=Off 1=On"}))
  Saturate(None=@n59)
  Saturate(None=@n177)
  Saturate(None=OneMinus())
  n279 = WorldPosition()
  n280 = Add(
   A=VectorParameter() {group=Toon_Shading, parameterName=SocketOffset, defaultValue=<rgba 0 0 0 1>}.RGB
   B=VectorParameter() {group=Toon_Shading, parameterName=HeadSocketLocation, defaultValue=<rgba 0 0 0 1>}.RGB
  )
  n281 = Multiply(
   A=Normalize(VectorInput=Subtract(A=@n279.XYZ, B=@n280))
   B=ScalarParameter() {defaultValue=1, group=Toon_Shading, parameterName=SphericalNormalScale}
  )
  n282 = ScalarParameter() {defaultValue=0.5, group=Toon_Shading, parameterName=EF_Anime_SphericalNormal_Blend_Amt}
  StaticSwitchParameter(True=Normalize(VectorInput=LinearInterpolate(A=@n133, B=@n281, Alpha=@n282)), False=@n133) {group=Toon_Shading, parameterName=EF_Anime_UseSphericalNormalBlend, sortPriority=0}
  n283 = Multiply(
   A=Multiply(
    A=NamedRerouteUsage()
    B=VectorParameter() {group=Effect_Anime_Base_Color, parameterName=EF_Anime_SpecularColor, defaultValue=<rgba 1 1 1 1>}.RGB
   )
   B=ScalarParameter() {group=Effect_Anime_Base_Color, parameterName=EF_Anime_SpecularStrength}
  )
  n284 = LinearInterpolate(Alpha=NamedRerouteUsage()) {constA=1, constB=0.5}
  StaticSwitchParameter(True=LinearInterpolate(A=@n81, B=Multiply(A=@n283, B=@n284) {constB=0.25}, Alpha=@n89.Result), False=@n81) {group=Effect_Anime_Base_Color, parameterName=EF_Anime_Specular__SpecularHighlight Use}
  StaticSwitchParameter(True=ChannelMaskParameter(None=NamedRerouteUsage()) {group=Toon_Shading_Fake PBR, parameterName=Ignore Z SkinFX Channel}.R, False=@n267) {group=Toon_Shading, parameterName=Ignore Z use SkinFX mask, sortPriority=0}
  n285 = StaticSwitchParameter(True=Custom() {code="return View.DirectionalLightDirection;", description=DirectionalLightDirection, inputs []}, False=@n141) {
   desc=use the Light vector of the first directional light in the scene instead of the Sun Vector.
   group=Effect_Anime
   parameterName=EF_Anime_UseDirectionalLight
   sortPriority=0
  }
  n286 = MaterialFunctionCall(Float3=@n285) {materialFunction=&BreakOutFloat3Components}
  n287 = Constant()
  StaticSwitchParameter(
   True=Normalize(VectorInput=MaterialFunctionCall(X=@n286.R, Y=@n286.R, Z=@n287) {materialFunction=&MakeFloat3}.Result) {desc=need to transform it to preskinning}
   False=@n285
  ) {desc=Lock the Z axis of the Light Vector, group=Toon_Shading, parameterName=UseLockLightZVector, sortPriority=0}
  n288 = TextureSampleParameter2D() {
   desc=ShadowGradient Texture
   group=Toon_Shading
   parameterName=SG_Texture
   samplerType=SAMPLERTYPE_LinearGrayscale
   sortPriority=1
   texture=/Game/Global/Textures/Default/Blanks/T_BlackLinearGrayscale
  }
  n289 = Multiply(A=@n288.R) {constB=2}
  n290 = MaterialFunctionCall(Float3=Transform(None=PixelNormalWS()) {transformSourceType=TRANSFORMSOURCE_World, transformType=TRANSFORM_Local}) {materialFunction=&BreakOutFloat3Components}
  n291 = Normalize(VectorInput=MaterialFunctionCall(X=Subtract(A=@n289), Y=@n290.G, Z=@n290.G) {materialFunction=&MakeFloat3}.Result)
  StaticSwitchParameter(
   True=LinearInterpolate(
    A=@n133
    B=Transform(None=@n291) {transformSourceType=TRANSFORMSOURCE_Local}
    Alpha=ScalarParameter() {defaultValue=0.5, group=Toon_Shading, parameterName=EF_Anime_ShadowGradient_Blend, sliderMax=1}
   )
   False=@n133
  ) {group=Toon_Shading, parameterName=UseShadowGradientMap, sortPriority=0}
  StaticSwitchParameter(
   True=Multiply(A=MaterialFunctionCall() {materialFunction=&CameraDirectionVector}.Result) {constB=-1}
   False=@n83
  ) {group=Toon_Shading, parameterName=UseCameraDirectionForSpecularHighlights, sortPriority=0}
  n292 = TextureSampleParameter2D() {group=Toon_Shading, parameterName=SSC_Texture_Far, sortPriority=1, texture=&T_BlacksRGB}
  n293 = Distance(A=WorldPosition().XYZ, B=CameraPositionWS())
  n294 = ScalarParameter() {defaultValue=100, group=Toon_Shading_SSC, parameterName=SSC_FarDistance}
  n295 = ScalarParameter() {defaultValue=500, group=Toon_Shading_SSC, parameterName=SSC_NearDistance}
  n296 = ComponentMask(None=ViewProperty() {property=MEVP_BufferSize}.Property) {r=true}
  n297 = ScalarParameter() {defaultValue=3840, group=Toon_Shading_SSC, parameterName=Reference Screen Size}
  n298 = Divide(A=Divide(A=Subtract(A=@n293, B=@n294), B=Subtract(A=@n295, B=@n294)), B=Power(Base=Divide(A=@n296, B=@n297)) {constExponent=1.2})
  StaticSwitchParameter(True=LinearInterpolate(A=@n132.RGBA, B=@n292.RGBA, Alpha=Saturate(None=@n298)), False=@n132.RGBA) {group=Toon_Shading, parameterName=Enable Far SSC, sortPriority=0}
  n299 = TextureSampleParameter2D(UVs=TextureCoordinate() {coordinateIndex=1}) {
   desc=Only overrides green channel in original SSC
   group=Toon_Shading
   parameterName=SSC_Texture_2nd_UV
   sortPriority=1
   texture=&T_BlacksRGB
  }
  n300 = ChannelMaskParameter(None=@n299.RGBA) {group=Toon_Shading_SSC, maskChannel=Green, parameterName=SSC_Texture_2nd_UV_Channel, defaultValue=<rgba 0 1 0 0>}
  StaticSwitchParameter(
   True=MaterialFunctionCall(X=@n132.R, Y=@n300.R, Z=@n132.R, A=@n132.R) {materialFunction=&MakeFloat4}.Result
   False=@n132.RGBA
  ) {desc=Doesn't work with Enable Far SSC, group=Toon_Shading, parameterName=Enable 2nd UV Channel SSC, sortPriority=0}
  TextureObject() {samplerType=SAMPLERTYPE_Masks, texture=&T_Scar_CM}
  TextureSample!(
   UVs
   Tex=NamedRerouteUsage()
   Level=ScalarParameter() {defaultValue=2, group=Effect_Anime, parameterName=EF_Anime_Curvature_mip_levels, sliderMax=1, sortPriority=50}
   Apply View MipBias
  ) {mipValueMode=TMVM_MipLevel}
  TextureSample!(UVs, Tex, Apply View MipBias) {texture=/Game/Weapons/FORT_Rifles/Textures/SCAR_D}
  VertexNormalWS()
 ]}
