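# Material function for weapon wrap skins that pulses a material's base color and emissive with in-game audio, behind a static switch.
# Event audio amplitude and emote low audio from the Fortnite parameter collection are remapped, then combined with emote high audio into one pulse value.
# That value lerps min and max strengths that multiply each channel, with separate on/off blends.
# A customization mask, which a parameter can override, limits where the effect applies.
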MaterialFunction /Game/Weapons/WeaponSkins/Wraps/MaterialFunctions/MF_Effect_Audio_Pulse
graph {nodes [
  n1 = FunctionInput() {inputName=Material, inputType=FunctionInput_MaterialAttributes}
  n2 = GetMaterialAttributes(None=@n1) {attributeGetTypes [#1, #2]}
  n3 = CollectionParameter() {collection=&FortniteMaterialParameters, parameterName=EventAudio_Amplitude}
  n4 = CollectionParameter() {collection=&FortniteMaterialParameters, parameterName=Emote_AudioLow}
  n5 = MaterialFunctionCall(
   Input=Max(A=@n3, B=@n4)
   Input Low=ScalarParameter() {group=Effect_AudioPulse, parameterName=AuidoPulse_Range_Low_Remap, sortPriority=40}
   Input High=ScalarParameter() {defaultValue=1, group=Effect_AudioPulse, parameterName=AuidoPulse_Range_High_Remap, sortPriority=40}
   Target Low=Constant()
   Target High=Constant() {r=1}
  ) {materialFunction=&RemapValueRange}
  n6 = CollectionParameter() {collection=&FortniteMaterialParameters, parameterName=Emote_AudioHigh}
  n7 = Saturate(None=Max(A=Saturate(None=@n5.Result), B=@n6))
  n8 = LinearInterpolate(
   A=ScalarParameter() {group=Effect_AudioPulse, parameterName=AuidoPulse_Min_BaseColor_strength, sortPriority=10}
   B=ScalarParameter() {defaultValue=1, group=Effect_AudioPulse, parameterName=AuidoPulse_Max_BaseColo_strength, sortPriority=10}
   Alpha=@n7
  )
  n9 = LinearInterpolate(
   A=ScalarParameter() {group=Effect_AudioPulse, parameterName=AuidoPulse_Min_Emissive_strength, sortPriority=20}
   B=ScalarParameter() {defaultValue=1, group=Effect_AudioPulse, parameterName=AuidoPulse_Max_Emissive_strength, sortPriority=20}
   Alpha=@n7
  )
  n10 = SetMaterialAttributes!(
   MaterialAttributes=@n2.MaterialAttributes
   BaseColor=LinearInterpolate(
    A=@n2.BaseColor
    B=Multiply(A=@n2.BaseColor, B=@n8)
    Alpha=ScalarParameter() {group=Effect_AudioPulse, parameterName="AuidoPulse_Effect_BaseColor 0=off  1=Active"}
   )
   EmissiveColor=LinearInterpolate(
    A=@n2.EmissiveColor
    B=Multiply(A=@n2.EmissiveColor, B=@n9)
    Alpha=ScalarParameter() {group=Effect_AudioPulse, parameterName="AuidoPulse_Effect_Emissive 0=off  1=Active", sortPriority=15}
   )
  ) {attributeSetTypes [#1, #2]}
  n11 = Saturate(None=Add(
    A=FunctionInput() {bUsePreviewValueAsDefault=true, inputName="Customization Mask (_CM B)", inputType=FunctionInput_Scalar, sortPriority=24}
    B=ScalarParameter() {group=Effect_AudioPulse, parameterName=AudioPulse_Remove_Customization Mask, sortPriority=-10}
   ))
  FunctionOutput(None=StaticSwitchParameter(True=BlendMaterialAttributes(A=@n2.MaterialAttributes, B=@n10, Alpha=@n11), False=@n1) {group=Customization, parameterName=Effect_Audio_Pulse})
 ]}
