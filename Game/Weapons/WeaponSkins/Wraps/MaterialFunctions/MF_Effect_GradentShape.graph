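# Material function for weapon wraps that adds an animated gradient-shape effect, toggled by a static switch and blended over the input material by masks.
# Projected UVs, optionally pixelated, sample shape and frame textures that drive curve-atlas gradients and time-of-day emissive.
# A custom hexagon SDF with warped rotation and event/emote audio parameters supplies pulsing shapes and colours.
# IsState inputs select tiling, offset and scale for weapon, vehicle, mech, character, backpack or pickaxe.
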MaterialFunction /Game/Weapons/WeaponSkins/Wraps/MaterialFunctions/MF_Effect_GradentShape
graph {nodes [
  n1 = NamedRerouteUsage()
  n2 = Multiply(A=Time(), B=ScalarParameter() {group=Effect_GradentShapes_Rotation_UVs, parameterName=GradShapes_Rotation_Warp_Speed_All})
  n3 = Add(A=NamedRerouteUsage()) {constB=-0.5}
  n4 = OneMinus(None=MaterialFunctionCall(Vector 3=@n3, Vector 2=@n3) {materialFunction=/Engine/Functions/Engine_MaterialFunctions02/Utility/VectorLength}.V2 Length)
  n5 = Multiply(A=@n4, B=ScalarParameter() {defaultValue=4, group=Effect_GradentShapes_Rotation_UVs, parameterName=GradShapes_Rotation_Warp_TileRate}) {constB=2}
  n6 = Multiply(A=Time(), B=ScalarParameter() {defaultValue=0.5, group=Effect_GradentShapes_Rotation_UVs, parameterName=GradShapes_Rotation_Warp_Speed})
  n7 = LinearInterpolate(
   A=ScalarParameter() {group=Effect_GradentShapes_Rotation_UVs, parameterName=GradShapes_Rotation_Warp_Amouint_outer}
   B=ScalarParameter() {defaultValue=1, group=Effect_GradentShapes_Rotation_UVs, parameterName=GradShapes_Rotation_Warp_Amouint}
   Alpha=Power(Base=@n4, Exp=ScalarParameter() {defaultValue=3, group=Effect_GradentShapes_Rotation_UVs, parameterName=GradShapes_Rotation_Warp_Falloff}) {constExponent=0.25}
  )
  n8 = NamedRerouteUsage()
  n9 = NamedRerouteUsage()
  n10 = Multiply(A=Sine(None=Add(A=@n5, B=@n6)), B=LinearInterpolate(A=@n7, B=Multiply(A=@n7, B=@n8), Alpha=Ceil(None=@n9))) {constB=10}
  MaterialExpressionAdd_10 = Add(A=Rotator(Coordinate=@n1, Time=Add(A=@n2, B=@n10))) {constB=-0.5}
  MaterialExpressionConstant_4 = Constant() {r=1}
  n11 = Custom!(p=@MaterialExpressionAdd_10, r=@MaterialExpressionConstant_4) {
   code="\r\n{\r\n    const float3 k = float3(-0.866025404,0.5,0.577350269);\r\n    p = abs(p);\r\n    p -= 2.0*min(dot(k.xy,p),0.0)*k.xy;\r\n    p -= float2(clamp(p.x, -k.z*r, k.z*r), r);\r\n    return length(p)*sign(p.y);\r\n}"
   outputType=CMOT_Float1
   inputs [
    {inputName=p, input {expression=&self:MaterialExpressionAdd_10}}
    {inputName=r, input {expression=&self:MaterialExpressionConstant_4}}
   ]
  }
  n12 = Abs(None=@n11)
  n13 = Power(Base=@n12, Exp=ScalarParameter() {defaultValue=1, group=Effect_GradentShapes_Hex, parameterName=GradShapes_Hex_Falloff})
  n14 = Multiply(A=@n13, B=NamedRerouteUsage()) {constB=0.25}
  n15 = NamedRerouteUsage()
  n16 = CurveAtlasRowParameter(CurveTime=Frac(None=Add(A=@n14, B=@n15))) {
   atlas=/Game/Weapons/WeaponSkins/Wraps/Textures/CruveAtlas_Wraps_01
   curve=/Game/Weapons/WeaponSkins/Wraps/Textures/CactusDancer/CC_Cactus_A
   defaultValue=73
   group=Effect_GradentShapes
   parameterName=GradShapes_CC_Gradent
  }
  AppendVector(A=@n16, B=@n16)
  n17 = TextureSample() {samplerType=SAMPLERTYPE_Masks, texture=/Game/Weapons/FORT_Rifles/Textures/T_Scar_CM}
  n18 = FunctionInput(Preview=@n17.R) {bUsePreviewValueAsDefault=true, inputName=Mask Primary, inputType=FunctionInput_Scalar, sortPriority=24, previewValue=<1 0 0 1>}
  n19 = FunctionInput(Preview=@n17.B) {bUsePreviewValueAsDefault=true, inputName=Mask Secondary, inputType=FunctionInput_Scalar, sortPriority=24, previewValue=<1 0 0 1>}
  n20 = Add(A=@n18, B=@n19)
  n21 = ScalarParameter() {group=Effect_GradentShapes, parameterName=GradShapes_Remove_Mask, sortPriority=0}
  n22 = BlendMaterialAttributes(A=NamedRerouteUsage(), B=NamedRerouteUsage(), Alpha=Saturate(None=Add(A=@n20, B=@n21))) {vertexAttributeBlendType=UseA}
  BlendMaterialAttributes(A=@n22, Alpha=NamedRerouteDeclaration(None=@n19) {name=Accent_Msk, nodeColor=<rgba 0 0.6484 1 1>}) {vertexAttributeBlendType=UseA}
  Constant4Vector()
  Constant()
  n23 = TextureSample() {samplerType=SAMPLERTYPE_Normal, texture=/Game/Weapons/FORT_Rifles/T_SCAR_N}
  n24 = Constant3Vector() {constant=<rgba 1 0 0 0>}
  n25 = GetMaterialAttributes(None=FunctionInput(Preview=SetMaterialAttributes!(MaterialAttributes, Normal=@n23.RGB, BaseColor=@n24) {attributeSetTypes [#1, #2]}) {inputName=Material, inputType=FunctionInput_MaterialAttributes}) {attributeGetTypes [
    #2
    #3
    #1
    #4
   ]}
  FunctionOutput(None=StaticSwitchParameter(True=@n22, False=NamedRerouteDeclaration(None=@n25.MaterialAttributes) {name=Root_Material, nodeColor=<rgba 0.7812 1 0 1>}) {desc=TopSide Projection Base - Emissive pan Mask, group=Customization, parameterName=Effect_GradentShapes})
  n26 = ScalarParameter() {defaultValue=73, group=Effect_GradentShapes, parameterName="GradShapes_CC_Gradent  Index"}
  n27 = Multiply(A=NamedRerouteUsage(), B=ScalarParameter() {group=Effect_GradentShapes, parameterName=GradShapes_Crevis_Wrap_Amount, sortPriority=0})
  n28 = MaterialFunctionCall(Float4=NamedRerouteUsage()) {materialFunction=&BreakOutFloat4Components}
  n29 = NamedRerouteUsage()
  n30 = MaterialFunctionCall!(
   For_Safe_Zones=StaticBool() {value=true}
   Normal_Map
   Caculated_Bounds=NamedRerouteUsage()
   Pre_Skinned_Normals=NamedRerouteUsage()
   Is Vehicle=@n28.R
   Is VehicleSide=@n28.R
   Is Mech=@n28.R
   Vehicle_Rotation
   UV_Rotation Bypass=NamedRerouteUsage()
   Mech_Rotation
   Uvs_Offset=NamedRerouteUsage()
   Offset-after rotation
   Uvs_TileRate=@n29
  ) {materialFunction=&/MF_Projection_Side_Normal}
  n31 = ScalarParameter() {defaultValue=100, group=Effect_GradentShapes, parameterName=GradShapes_Pixel_Density}
  n32 = Multiply(A=@n30.UVs, B=@n31)
  n33 = LinearInterpolate(
   A=Divide(A=Floor(None=@n32), B=@n31)
   B=@n30.UVs
   Alpha=ScalarParameter() {group=Effect_GradentShapes, parameterName="GradShapes_Pixel_UVS  1=Off 0=On"}
  )
  n34 = ChannelMaskParameter(None=TextureSample(UVs=Add(A=@n27, B=@n33), Tex=NamedRerouteUsage()) {samplerType=SAMPLERTYPE_LinearGrayscale}.RGBA) {group=Effect_GradentShapes, maskChannel=Alpha, parameterName=GradShapes__Tex_Shape_Channel, sortPriority=5, defaultValue=<rgba 0 0 0 1>}
  n35 = ScalarParameter() {defaultValue=1, group=Effect_GradentShapes, parameterName=GradShapes_CC_TileRate}
  n36 = Frac(None=Add(A=Multiply(A=@n34.R, B=@n35) {constB=3}, B=NamedRerouteUsage()))
  n37 = CurveAtlasRowParameter(CurveTime=@n36) {
   atlas=&CruveAtlas_Wraps_01
   curve=&CC_Cactus_A
   defaultValue=73
   group=Effect_GradentShapes
   parameterName=GradShapes_CC_Gradent
  }
  n38 = MaterialFunctionCall!(Index=@n26, CCinput=@n37, CCinput_Alpha, UV=@n36) {materialFunction=&/MF_Global_CC_Tex}
  LinearInterpolate(A=@n38.3V)
  LinearInterpolate()
  MaterialFunctionCall(Float2=@n1) {materialFunction=&BreakOutFloat2Components}
  n39 = CollectionParameter() {collection=&FortniteMaterialParameters, parameterName=EventAudio_Spectrum}
  n40 = MaterialFunctionCall(Float4=@n39) {materialFunction=&BreakOutFloat4Components}
  n41 = Multiply(A=@n40.A, B=ScalarParameter() {defaultValue=1, group=Effect_GradShapes_Audio, parameterName=GradShapes_EventAudioMultiplier_lines})
  Min(A=@n41) {constB=30, desc=prevent crazy values}
  Multiply()
  n42 = Add(A=@n1) {constB=-0.5}
  n43 = Multiply(A=Arctangent2!(Y=ComponentMask(None=@n42) {r=true}, X=ComponentMask(None=@n42) {g=true}))
  Multiply(A=TextureSample(UVs=AppendVector(A=@n12, B=@n43)) {texture=/Game/Weapons/WeaponSkins/Wraps/Textures/UVCheaker}.RGB)
  n44 = NamedRerouteUsage()
  Multiply(
   A=Fresnel(ExponentIn=ScalarParameter() {defaultValue=2, group=Effect_GradShapes_Audio, parameterName=GradShapes_BodyFresnelExp, sliderMax=8}) {baseReflectFraction=0, exponent=2}
   B=LinearInterpolate(
    A=ScalarParameter() {defaultValue=0.1, group=Effect_GradShapes_Audio, parameterName=GradShapes_BodyFresnel_Intensity, sliderMax=1}
    B=ScalarParameter() {defaultValue=0.5, group=Effect_GradShapes_Audio, parameterName=GradShapes_BodyFresnel_Intensity_Active, sliderMax=1}
    Alpha=Saturate(None=@n44)
   )
  )
  n45 = ScalarParameter() {defaultValue=1, group=Effect_GradShapes_Audio, parameterName=GradShapes_EventAudioMultiplier}
  n46 = Min(A=Multiply(A=@n45, B=ScalarParameter() {defaultValue=1, group=Effect_GradShapes_Audio, parameterName=GradShapes_EventAudioMultiplier_Hex})) {constB=30, desc=prevent crazy values}
  Multiply(A=MaterialFunctionCall!(
    X range Y range
    Gradient Height
    ShowGrid?
    Thickness
    Color
    "f(x)"=NamedRerouteDeclaration(None=@n46) {name=AudioSpectrum_2, nodeColor=<rgba 1 0.8203 0 1>}
   ) {materialFunction=/Engine/Functions/Engine_MaterialFunctions01/Debug/PlotFunctionOnGraph}."f(x) = y")
  n47 = Multiply(A=ScalarParameter() {defaultValue=4, group=Effect_GradentShapes_Hex, parameterName=GradShapes_Hex_Quantize}, B=NamedRerouteUsage()) {constB=4}
  n48 = Divide(A=Floor(None=Multiply(A=@n13, B=@n47)), B=@n47)
  n49 = Power(Base=@n48, Exp=ScalarParameter() {defaultValue=1, group=Effect_GradentShapes_Hex, parameterName=GradShapes_Hex_Quantize_pow})
  Multiply(A=@n49)
  Multiply(A=MaterialFunctionCall("f(x)"=@n49) {materialFunction=&PlotFunctionOnGraph}."f(x) = y")
  n50 = NamedRerouteUsage()
  n51 = ScalarParameter() {defaultValue=0.1, group=Effect_GradentShapes_Hex, parameterName=GradShapes_Hex_Color_Offset}
  n52 = MaterialFunctionCall(
   A=VectorParameter() {group=Effect_GradentShapes_Hex, parameterName=GradShapes_Hex_Color_A, defaultValue=<rgba 0 0.0462 1 1>}.RGB
   B=VectorParameter() {group=Effect_GradentShapes_Hex, parameterName=GradShapes_Hex_Color_B, defaultValue=<rgba 0.1499 1 0 1>}.RGB
   C=VectorParameter() {group=Effect_GradentShapes_Hex, parameterName=GradShapes_Hex_Color_C, defaultValue=<rgba 1 0.0553 0 1>}.RGB
   Alpha=Power(
    Base=Add(A=@n50, B=@n51) {constB=0.1}
    Exp=ScalarParameter() {defaultValue=1, group=Effect_GradentShapes_Hex, parameterName=GradShapes_Hex_Color_Falloff}
   ) {constExponent=4}
  ) {materialFunction=&3ColorBlend}
  Multiply(A=@n52.Result)
  n53 = TextureObjectParameter() {group=Effect_GradentShapes, parameterName=GradShapes__Tex_Shape, sortPriority=5, texture=&T_BlacksRGB}
  NamedRerouteDeclaration(None=@n53) {name=Tex_Shape, nodeColor=<rgba 0 0.4609 1 1>}
  n54 = TextureObjectParameter() {group=Effect_GradentShapes, parameterName=GradShapes__Tex_Frame, sortPriority=5, texture=&T_BlacksRGB}
  NamedRerouteDeclaration(None=@n54) {name=Tex_Frame, nodeColor=<rgba 0 0.4609 1 1>}
  n55 = MaterialFunctionCall(Color=Abs(None=NamedRerouteUsage())) {materialFunction=&SplitComponents}
  n56 = MaterialFunctionCall(Float4=NamedRerouteUsage()) {materialFunction=&BreakOutFloat4Components}
  n57 = LinearInterpolate(A=@n55.B, B=@n55.B, Alpha=@n56.G)
  n58 = MaterialFunctionCall(
   In=LinearInterpolate(A=LinearInterpolate(A=@n55.R, B=@n57, Alpha=@n56.R), B=@n55.G, Alpha=@n56.B)
   Contrast=ScalarParameter() {defaultValue=5, group=Effect_GradentShapes, parameterName=GradShapes__TriPlanerMsk_Contrast, sortPriority=0}
  ) {materialFunction=&CheapContrast}
  n59 = LinearInterpolate(
   A=ScalarParameter() {group=Effect_GradentShapes, parameterName="GradShapes__TriPlanerMsk 0=On", sortPriority=0}
   B=ScalarParameter() {group=Effect_GradentShapes, parameterName="GradShapes__TriPlanerMsk Char 0=On", sortPriority=0}
   Alpha=@n56.A
  )
  NamedRerouteDeclaration(None=Saturate(None=Add(A=@n58.Result, B=@n59))) {name=Tri_Mask, nodeColor=<rgba 1 0.4688 0 1>}
  NamedRerouteDeclaration(None=MaterialFunctionCall(Index=@n26, CCinput=@n16, CCinput_Alpha=@n16) {materialFunction=&MF_Global_CC_Tex}.4v) {name=ColorCurve, nodeColor=<rgba 0.1016 0 1 1>}
  NamedRerouteDeclaration(None=@n30.UVs) {name=Uvs, nodeColor=<rgba 0.8984 1 0 1>}
  n60 = ScalarParameter() {defaultValue=1, group=Effect_GradentShapes_Hex, parameterName=GradShapes_Hex_TileRater_Vehicle, sortPriority=150}
  n61 = ScalarParameter() {defaultValue=1, group=Effect_GradentShapes_Hex, parameterName=GradShapes_Hex_TileRater_Vehicle_Side, sortPriority=150}
  n62 = MaterialFunctionCall(Float4=NamedRerouteUsage()) {materialFunction=&BreakOutFloat4Components}
  n63 = LinearInterpolate(
   A=LinearInterpolate(
    A=ScalarParameter() {defaultValue=5, group=Effect_GradentShapes_Hex, parameterName=GradShapes_Hex_TileRater_Weapon, sortPriority=150}
    B=LinearInterpolate(A=@n60, B=@n61, Alpha=@n62.G)
    Alpha=@n62.R
   )
   B=LinearInterpolate(
    A=ScalarParameter() {defaultValue=1, group=Effect_GradentShapes_Hex, parameterName=GradShapes_Hex_TileRater_Mech, sortPriority=150}
    B=ScalarParameter() {defaultValue=1, group=Effect_GradentShapes_Hex, parameterName=GradShapes_Hex_TileRater_Character, sortPriority=150}
    Alpha=@n62.A
   )
   Alpha=@n62.B
  )
  n64 = ScalarParameter() {defaultValue=1, group=Effect_GradentShapes_Hex, parameterName=GradShapes_Hex_TileRater_Backpack, sortPriority=150}
  n65 = MaterialFunctionCall(Float2=NamedRerouteUsage()) {materialFunction=&BreakOutFloat2Components}
  NamedRerouteDeclaration(None=LinearInterpolate(
    A=LinearInterpolate(A=@n63, B=@n64, Alpha=@n65.R)
    B=ScalarParameter() {defaultValue=1, group=Effect_GradentShapes_Hex, parameterName=GradShapes_Hex_TileRater_Pickaxe, sortPriority=150}
    Alpha=@n65.G
   )) {name=UV_GradShapes_Hex_TileRate, nodeColor=<rgba 0 0.7188 1 1>}
  n66 = NamedRerouteUsage()
  NamedRerouteDeclaration(None=SmoothStep(Min=@n66, Max=@n66, Value=@n11) {constMin=0.6}) {name=Hex Center Blockout, nodeColor=<rgba 0.6641 0 1 1>}
  n67 = ScalarParameter() {defaultValue=-0.95, group=Effect_GradentShapes_Hex, parameterName=GradShapes_Hex_SizeBB_Vehicle, sortPriority=150}
  n68 = ScalarParameter() {defaultValue=-0.95, group=Effect_GradentShapes_Hex, parameterName=GradShapes_Hex_SizeBB_Vehicle_Side, sortPriority=150}
  n69 = MaterialFunctionCall(Float4=NamedRerouteUsage()) {materialFunction=&BreakOutFloat4Components}
  n70 = LinearInterpolate(
   A=LinearInterpolate(
    A=ScalarParameter() {defaultValue=-0.95, group=Effect_GradentShapes_Hex, parameterName=GradShapes_Hex_SizeBB_Weapon, sortPriority=150}
    B=LinearInterpolate(A=@n67, B=@n68, Alpha=@n69.G)
    Alpha=@n69.R
   )
   B=LinearInterpolate(
    A=ScalarParameter() {defaultValue=-0.95, group=Effect_GradentShapes_Hex, parameterName=GradShapes_Hex_SizeBB_Mech, sortPriority=150}
    B=ScalarParameter() {defaultValue=-0.95, group=Effect_GradentShapes_Hex, parameterName=GradShapes_Hex_SizeBB_Character, sortPriority=150}
    Alpha=@n69.A
   )
   Alpha=@n69.B
  )
  n71 = ScalarParameter() {defaultValue=-0.95, group=Effect_GradentShapes_Hex, parameterName=GradShapes_Hex_SizeBB_Backpack, sortPriority=150}
  n72 = MaterialFunctionCall(Float2=NamedRerouteUsage()) {materialFunction=&BreakOutFloat2Components}
  NamedRerouteDeclaration(None=LinearInterpolate(
    A=LinearInterpolate(A=@n70, B=@n71, Alpha=@n72.R)
    B=ScalarParameter() {defaultValue=-0.95, group=Effect_GradentShapes_Hex, parameterName=GradShapes_Hex_SizeBB_Pickaxe, sortPriority=150}
    Alpha=@n72.G
   )) {name=UV_Hex_BB_Size, nodeColor=<rgba 0 0.7188 1 1>}
  n73 = GetMaterialAttributes(None=NamedRerouteUsage()) {attributeGetTypes [#2, #5]}
  n74 = NamedRerouteUsage()
  n75 = ChannelMaskParameter(None=NamedRerouteUsage()) {group=Effect_GradentShapes, parameterName=GradShapes_CC_Masking, sortPriority=5}
  n76 = NamedRerouteUsage()
  n77 = Multiply(A=Multiply(A=@n74, B=@n75.R), B=@n76)
  n78 = NamedRerouteUsage()
  n79 = Multiply(A=OneMinus(None=@n75.R), B=Multiply(A=@n74, B=Multiply(A=@n76, B=@n78)))
  n80 = OneMinus(None=Saturate(None=Add(A=@n77, B=@n79)))
  n81 = Multiply(A=@n73.BaseColor, B=@n80)
  n82 = Multiply(
   A=Multiply(A=@n73.BaseColor, B=@n77)
   B=ScalarParameter() {defaultValue=1, group=Effect_GradentShapes CC_Msking, parameterName=GradShapes_CC_Masking_Brightness}
  )
  n83 = Multiply(
   A=@n52.Result
   B=ScalarParameter() {defaultValue=1, group=Effect_GradentShapes CC_Msking, parameterName=GradShapes_CC_Masking_Brightness_Audio}
  )
  n84 = Saturate(None=@n79)
  n85 = Add(
   A=@n73.EmissiveColor
   B=Desaturation(
    None=LinearInterpolate(A=@n82, B=@n83, Alpha=@n84)
    Fraction=ScalarParameter() {group=Effect_GradentShapes CC_Msking, parameterName=GradShapes_CC_Masking_De}
   )
  )
  n86 = VectorParameter() {group=Effect_GradentShapes, parameterName=GradShapes_BaseColor, sortPriority=10, defaultValue=<rgba 0 0 0 1>}
  n87 = Add(A=Multiply(A=NamedRerouteUsage(), B=Divide(A=@n31, B=@n29)), B=@n27)
  n88 = NamedRerouteUsage()
  n89 = Add(A=Multiply(A=@n30.UVs, B=@n31), B=@n27)
  n90 = ChannelMaskParameter(None=LinearInterpolate(
    A=TextureSample(UVs=@n87, Tex=@n88) {samplerType=SAMPLERTYPE_LinearGrayscale}.RGBA
    B=TextureSample(UVs=@n89, Tex=@n88) {samplerType=SAMPLERTYPE_LinearGrayscale}.RGBA
    Alpha=NamedRerouteUsage()
   )) {group=Effect_GradentShapes, parameterName=GradShapes__Tex_Frame_Channel, sortPriority=5}
  n91 = Multiply(
   A=Multiply(A=@n38.3V, B=@n90.R)
   B=MaterialFunctionCall(
    Value_Morning=ScalarParameter() {defaultValue=0.5, group=Effect_GradentShapes, parameterName=GradShapes_TOD_Value_Morning, sortPriority=10}
    Value_Day=ScalarParameter() {defaultValue=1, group=Effect_GradentShapes, parameterName=GradShapes_TOD_Value_Day, sortPriority=10}
    Value_Evening=ScalarParameter() {defaultValue=0.5, group=Effect_GradentShapes, parameterName=GradShapes_TOD_Value_Evening, sortPriority=10}
    Value_Night=ScalarParameter() {defaultValue=0.25, group=Effect_GradentShapes, parameterName=GradShapes_TOD_Value_Night, sortPriority=10}
   ) {materialFunction=&Scalar_Time_of_Day}.Result
  )
  n92 = Multiply(A=@n91, B=ScalarParameter() {defaultValue=1, group=Effect_GradentShapes, parameterName=GradShapes_Edge_Brightness, sortPriority=10}) {constB=5}
  n93 = NamedRerouteUsage()
  n94 = Saturate(None=Multiply(A=Add(A=@n93) {constB=-0.5}) {constB=2})
  n95 = Saturate(None=Multiply(A=@n93) {constB=2})
  n96 = ScalarParameter() {group=Effect_GradentShapes, parameterName=GradShapes_Edge_Shadows, sortPriority=10}
  n97 = SetMaterialAttributes!(
   MaterialAttributes=NamedRerouteUsage()
   BaseColor=@n86.RGB
   Roughness=LinearInterpolate(
    A=ScalarParameter() {group=Effect_GradentShapes, parameterName=GradShapes_Roughness_High, sortPriority=10}
    B=ScalarParameter() {defaultValue=1, group=Effect_GradentShapes, parameterName=GradShapes_Roughness_Low, sortPriority=10}
    Alpha=@n90.R
   )
   Specular=LinearInterpolate(
    A=ScalarParameter() {defaultValue=0.5, group=Effect_GradentShapes, parameterName=GradShapes_Specular_Low, sortPriority=10}
    B=ScalarParameter() {defaultValue=0.5, group=Effect_GradentShapes, parameterName=GradShapes_Specular_High, sortPriority=10}
    Alpha=@n90.R
   )
   Metallic=ScalarParameter() {group=Effect_GradentShapes, parameterName=GradShapes_Metallic, sortPriority=10}
   Normal
   EmissiveColor=Multiply(A=LinearInterpolate(A=@n91, B=@n92, Alpha=@n94), B=Subtract(A=@n95, B=@n96) {constB=0})
  ) {attributeSetTypes [
    #2
    #3
    #6
    #4
    #1
    #5
   ]}
  NamedRerouteDeclaration(None=StaticSwitchParameter(
    True=SetMaterialAttributes!(MaterialAttributes=@n73.MaterialAttributes, BaseColor=@n81, EmissiveColor=@n85) {attributeSetTypes [#2, #5]}
    False=@n97
   ) {group=Effect_GradentShapes, parameterName=GradShapes_Mask_BC_Add, sortPriority=0}) {name=Material_Back, nodeColor=<rgba 0 0.3906 1 1>}
  n98 = NamedRerouteUsage()
  n99 = OneMinus(None=Add(A=NamedRerouteUsage(), B=ScalarParameter() {group=Effect_GradShapes_Audio, parameterName=GradShapes_EventAudioMultiplier_A}) {constB=-0.25})
  NamedRerouteDeclaration(None=SmoothStep(Min=@n98, Max=Add(A=@n98) {constB=-0.01}, Value=Add(A=@n11, B=@n99) {constB=0}) {constMin=0.6}) {name=Hex Center Audio, nodeColor=<rgba 0.6641 0 1 1>}
  NamedRerouteDeclaration(None=@n41) {name=AudioSpectrum_1, nodeColor=<rgba 1 0.8203 0 1>}
  n100 = Frac(None=Multiply(A=Time(), B=ScalarParameter() {defaultValue=0.1, group=Effect_GradShapes_Audio, parameterName=GradShapes_Color_Speed}) {constB=0.5})
  n101 = CollectionParameter() {collection=&FortniteMaterialParameters, parameterName=Emote_AudioLow}
  n102 = Multiply(A=@n101, B=ScalarParameter() {defaultValue=1, group=Effect_GradShapes_Audio, parameterName=GradShapes_EventAudioRemoteMultiplier})
  n103 = CollectionParameter() {collection=&FortniteMaterialParameters, parameterName=EventAudio_Amplitude}
  n104 = Add(A=Add(A=@n40.R, B=@n40.R), B=Add(A=@n40.B, B=@n40.B))
  n105 = Multiply(A=Add(A=@n103, B=Divide(A=@n104) {constB=4}), B=@n45)
  n106 = NamedRerouteDeclaration(None=Max(A=Min(A=@n102) {constB=30, desc=prevent crazy values}, B=Min(A=@n105) {constB=30, desc=prevent crazy values})) {name=AudioSpectrum, nodeColor=<rgba 1 0.8203 0 1>}
  n107 = Add(
   A=@n100
   B=Multiply(A=ScalarParameter() {defaultValue=0.1, group=Effect_GradShapes_Audio, parameterName=GradShapes_Color_SpeedPulseStrength}, B=@n106) {constB=0.7}
  ) {constB=0.2}
  NamedRerouteDeclaration(None=LinearInterpolate(
    A=If(A=@n106, "A > B"=@n107, "A < B"=@n100) {constB=0.1}
    B=@n100
    Alpha=ScalarParameter() {group=Effect_GradShapes_Audio, parameterName="GradShapes_Color_Disable_Audio 1=On", sortPriority=0}
   )) {name=Audio_Speed, nodeColor=<rgba 0 0.3203 1 1>}
  n108 = MaterialFunctionCall(Float3=Abs(None=NamedRerouteUsage())) {materialFunction=&BreakOutFloat3Components}
  n109 = Constant() {r=0.5}
  n110 = NamedRerouteUsage()
  n111 = ComponentMask(None=@n110) {b=true, desc=YZ, g=true}
  n112 = ComponentMask(None=@n110) {b=true, desc=XZ, r=true}
  NamedRerouteDeclaration(None=If(
    A=@n108.B
    B=@n109
    "A > B"=ComponentMask(None=@n110) {desc=YX, g=true, r=true}
    "A < B"=If(A=@n108.R, B=@n109, "A > B"=@n111, "A < B"=@n112)
   )) {name=UVS_Cheap_Tri, nodeColor=<rgba 1 0 0.8438 1>}
  n113 = VectorParameter() {group=Effect_GradentShapes_UVs, parameterName=GradShapes_Offset_Vehicle, defaultValue=<rgba 0 0 0 1>}
  n114 = VectorParameter() {group=Effect_GradentShapes_UVs, parameterName=GradShapes_Offset_VehicleSide, defaultValue=<rgba 0 0 0 1>}
  n115 = MaterialFunctionCall(Float4=NamedRerouteUsage()) {materialFunction=&BreakOutFloat4Components}
  n116 = LinearInterpolate(
   A=LinearInterpolate(
    A=VectorParameter() {group=Effect_GradentShapes_UVs, parameterName=GradShapes_Offset_Weapon}.RGB
    B=LinearInterpolate(A=@n113.RGB, B=@n114.RGB, Alpha=@n115.G)
    Alpha=@n115.R
   )
   B=LinearInterpolate(
    A=VectorParameter() {group=Effect_GradentShapes_UVs, parameterName=GradShapes_Offset_Mech, defaultValue=<rgba 0 0 0 1>}.RGB
    B=VectorParameter() {group=Effect_GradentShapes_UVs, parameterName=GradShapes_Offset_Characte, defaultValue=<rgba 0 0 0 1>}.RGB
    Alpha=@n115.A
   )
   Alpha=@n115.B
  )
  n117 = VectorParameter() {group=Effect_GradentShapes_UVs, parameterName=GradShapes_Offset_Backpack, defaultValue=<rgba 0 0 0 1>}
  n118 = MaterialFunctionCall(Float2=NamedRerouteUsage()) {materialFunction=&BreakOutFloat2Components}
  NamedRerouteDeclaration(None=LinearInterpolate(
    A=LinearInterpolate(A=@n116, B=@n117.RGB, Alpha=@n118.R)
    B=VectorParameter() {group=Effect_GradentShapes_UVs, parameterName=GradShapes_Offset_Pickaxe, defaultValue=<rgba 0 0 0 1>}.RGB
    Alpha=@n118.G
   )) {name=UVs_offset, nodeColor=<rgba 0 0.625 1 1>}
  n119 = ScalarParameter() {group=Effect_GradentShapes_UVs, parameterName=GradShapes_Rotation_Weapon}
  n120 = MaterialFunctionCall(Float4=NamedRerouteUsage()) {materialFunction=&BreakOutFloat4Components}
  n121 = LinearInterpolate(
   A=ScalarParameter() {group=Effect_GradentShapes_UVs, parameterName=GradShapes_Rotation_Vehicle}
   B=ScalarParameter() {group=Effect_GradentShapes_UVs, parameterName=GradShapes_Rotation_VehicleSide}
   Alpha=@n120.G
  )
  n122 = ScalarParameter() {group=Effect_GradentShapes_UVs, parameterName=GradShapes_Rotation_Mech}
  n123 = ScalarParameter() {group=Effect_GradentShapes_UVs, parameterName=GradShapes_Rotation_Character}
  NamedRerouteDeclaration(None=LinearInterpolate(A=LinearInterpolate(A=@n119, B=@n121, Alpha=@n120.R), B=LinearInterpolate(A=@n122, B=@n123, Alpha=@n120.A), Alpha=@n120.B)) {name=UVs_Rotation, nodeColor=<rgba 1 0.7266 0 1>}
  NamedRerouteDeclaration(None=@n48) {name=Hex Center, nodeColor=<rgba 0.6641 0 1 1>}
  n124 = ScalarParameter() {defaultValue=1, group=Effect_GradentShapes_UVs, parameterName=GradShapes_Pixel_Density_Vehicle, sortPriority=150}
  n125 = ScalarParameter() {defaultValue=1, group=Effect_GradentShapes_UVs, parameterName=GradShapes_Pixel_Density_Vehicle_Side, sortPriority=150}
  n126 = MaterialFunctionCall(Float4=NamedRerouteUsage()) {materialFunction=&BreakOutFloat4Components}
  n127 = LinearInterpolate(
   A=LinearInterpolate(
    A=ScalarParameter() {defaultValue=1, group=Effect_GradentShapes_UVs, parameterName=GradShapes_Pixel_Density_Weapon, sortPriority=150}
    B=LinearInterpolate(A=@n124, B=@n125, Alpha=@n126.G)
    Alpha=@n126.R
   )
   B=LinearInterpolate(
    A=ScalarParameter() {defaultValue=1, group=Effect_GradentShapes_UVs, parameterName=GradShapes_Pixel_Density_Mech, sortPriority=150}
    B=ScalarParameter() {defaultValue=1, group=Effect_GradentShapes_UVs, parameterName=GradShapes_Pixel_Density_Character, sortPriority=150}
    Alpha=@n126.A
   )
   Alpha=@n126.B
  )
  n128 = ScalarParameter() {defaultValue=1, group=Effect_GradentShapes_UVs, parameterName=GradShapes_Pixel_Density_Backpack, sortPriority=150}
  n129 = MaterialFunctionCall(Float2=NamedRerouteUsage()) {materialFunction=&BreakOutFloat2Components}
  NamedRerouteDeclaration(None=LinearInterpolate(
    A=LinearInterpolate(A=@n127, B=@n128, Alpha=@n129.R)
    B=ScalarParameter() {defaultValue=1, group=Effect_GradentShapes_UVs, parameterName=GradShapes_Pixel_Density_Pickaxe, sortPriority=150}
    Alpha=@n129.G
   )) {name=UV_Pixel_Density, nodeColor=<rgba 0 0.7188 1 1>}
  n130 = ScalarParameter() {defaultValue=1, group=Effect_GradentShapes_UVs, parameterName=GradShapes_Scale_Vehicle, sortPriority=150}
  n131 = ScalarParameter() {defaultValue=1, group=Effect_GradentShapes_UVs, parameterName=GradShapes_Scale_Vehicle_Side, sortPriority=150}
  n132 = MaterialFunctionCall(Float4=NamedRerouteUsage()) {materialFunction=&BreakOutFloat4Components}
  n133 = LinearInterpolate(
   A=LinearInterpolate(
    A=ScalarParameter() {defaultValue=1, group=Effect_GradentShapes_UVs, parameterName=GradShapes_Scale_Weapon, sortPriority=150}
    B=LinearInterpolate(A=@n130, B=@n131, Alpha=@n132.G)
    Alpha=@n132.R
   )
   B=LinearInterpolate(
    A=ScalarParameter() {defaultValue=1, group=Effect_GradentShapes_UVs, parameterName=GradShapes_Scale_Mech, sortPriority=150}
    B=ScalarParameter() {defaultValue=1, group=Effect_GradentShapes_UVs, parameterName=GradShapes_Scale_Character, sortPriority=150}
    Alpha=@n132.A
   )
   Alpha=@n132.B
  )
  n134 = ScalarParameter() {defaultValue=1, group=Effect_GradentShapes_UVs, parameterName=GradShapes_Scale_Backpack, sortPriority=150}
  n135 = MaterialFunctionCall(Float2=NamedRerouteUsage()) {materialFunction=&BreakOutFloat2Components}
  NamedRerouteDeclaration(None=LinearInterpolate(
    A=LinearInterpolate(A=@n133, B=@n134, Alpha=@n135.R)
    B=ScalarParameter() {defaultValue=1, group=Effect_GradentShapes_UVs, parameterName=GradShapes_Scale_Pickaxe, sortPriority=150}
    Alpha=@n135.G
   )) {name=UV_Scale, nodeColor=<rgba 0 0.7188 1 1>}
  NamedRerouteDeclaration(None=FunctionInput(Preview=Constant4Vector().RGBA) {bUsePreviewValueAsDefault=true, inputName=IsState, inputType=FunctionInput_Vector4, sortPriority=4}) {name=IsState, nodeColor=<rgba 0.1497 1 0.1633 1>}
  NamedRerouteDeclaration(None=FunctionInput(Preview=Constant2Vector()) {bUsePreviewValueAsDefault=true, inputName=IsState_B, inputType=FunctionInput_Vector2, sortPriority=4}) {name=IsState_B, nodeColor=<rgba 0.1497 1 0.1633 1>}
  n136 = PreSkinnedLocalBounds()
  n137 = Subtract(A=VertexInterpolator(VS=PreSkinnedPosition()).PS, B=@n136.Min)
  n138 = MaterialFunctionCall(Color=Subtract(A=@n136.Max, B=@n136.Max)) {materialFunction=&SplitComponents}
  n139 = Max(A=Max(A=@n138.R, B=@n138.R), B=@n138.B)
  NamedRerouteDeclaration(None=FunctionInput(Preview=Divide(A=@n137, B=@n139)) {bUsePreviewValueAsDefault=true, inputName=Calcated_Bounds, sortPriority=11}) {name=Caclcated_Bounds, nodeColor=<rgba 0 0.9297 1 1>}
  n140 = FunctionInput(Preview=@n17.G) {bUsePreviewValueAsDefault=true, inputName=Mask Crevise, inputType=FunctionInput_Scalar, sortPriority=24, previewValue=<1 0 0 1>}
  NamedRerouteDeclaration(None=@n140) {name=Msk_Crevise, nodeColor=<rgba 0.5234 1 0 1>}
  n141 = PreSkinnedNormal()
  NamedRerouteDeclaration(None=FunctionInput(Preview=VertexInterpolator(VS=@n141).PS) {bUsePreviewValueAsDefault=true, inputName=Pre-Skinned Local Normal, sortPriority=10}) {name=PreSkin_Normal, nodeColor=<rgba 0.8281 1 0 1>}
  NamedRerouteDeclaration(None=@n25.Normal) {name=Root_Normal, nodeColor=<rgba 1 0.1641 0 1>}
  NamedRerouteUsage()
  Saturate(None=@n18)
  Saturate(None=@n80)
  Saturate(None=@n9)
  ScalarParameter() {defaultValue=0.5, group=Effect_GradentShapes, parameterName=GradShapes_Specular, sortPriority=10}
  VertexNormalWS()
 ]}
