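# Weapon wrap material function that hue-shifts and desaturates BaseColor and EmissiveColor of incoming material attributes.
# Each channel's shift sums a static offset, a time-animated offset and a reactive Start-to-End lerp timed from a timestamp parameter (named for Blueprint) over a duration.
# The result blends over the original by the customization mask plus a mask-removal parameter, saturated.
# The Effect_Hue_Shift static switch bypasses the effect when off.
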MaterialFunction /Game/Weapons/WeaponSkins/Wraps/MaterialFunctions/MF_Effect_Hue_Shift
graph {nodes [
  n1 = FunctionInput() {inputName=Material, inputType=FunctionInput_MaterialAttributes}
  n2 = GetMaterialAttributes(None=@n1) {attributeGetTypes [#1, #2]}
  n3 = Divide(
   A=Subtract(A=Time(), B=ScalarParameter() {defaultValue=-9999999, group=Effect_HueShift_Reactive, parameterName=HueShift_BaseColor_Time_BP_Brightness})
   B=ScalarParameter() {defaultValue=0.25, group=Effect_HueShift_Reactive, parameterName=HueShift_BaseColor_Time_BP_Duration_Brightness}
  )
  n4 = Time()
  n5 = ScalarParameter() {group=Effect_HueShift, parameterName=HueShift_BaseColor_Animated}
  n6 = Add(
   A=LinearInterpolate(
    A=ScalarParameter() {defaultValue=1, group=Effect_HueShift_Reactive, parameterName=HueShift_BaseColor_Start}
    B=ScalarParameter() {group=Effect_HueShift_Reactive, parameterName=HueShift_BaseColor_End}
    Alpha=Saturate(None=@n3)
   )
   B=Add(A=ScalarParameter() {group=Effect_HueShift, parameterName=HueShift_BaseColor}, B=Multiply(A=@n4, B=@n5))
  )
  n7 = Divide(
   A=Subtract(
    A=Time()
    B=ScalarParameter() {defaultValue=-9999999, group=Effect_HueShift_Reactive, parameterName=HueShift_EmissiveColor_Time_BP_Brightness}
   )
   B=ScalarParameter() {defaultValue=0.25, group=Effect_HueShift_Reactive, parameterName=HueShift_EmissiveColor_Time_BP_Duration_Brightness}
  )
  n8 = Time()
  n9 = ScalarParameter() {group=Effect_HueShift, parameterName=HueShift_EmissiveColor_Animated}
  n10 = Add(
   A=LinearInterpolate(
    A=ScalarParameter() {defaultValue=1, group=Effect_HueShift_Reactive, parameterName=HueShift_EmissiveColor_Start}
    B=ScalarParameter() {group=Effect_HueShift_Reactive, parameterName=HueShift_EmissiveColor_End}
    Alpha=Saturate(None=@n7)
   )
   B=Add(A=ScalarParameter() {group=Effect_HueShift, parameterName=HueShift_EmissiveColor}, B=Multiply(A=@n8, B=@n9))
  )
  n11 = SetMaterialAttributes!(
   MaterialAttributes=@n2.MaterialAttributes
   BaseColor=Desaturation(
    None=MaterialFunctionCall(Hue Shift Percentage=@n6, Texture=@n2.BaseColor) {materialFunction=&HueShift}.Result
    Fraction=ScalarParameter() {group=Effect_HueShift, parameterName=HueShift_BaseColorSaturation}
   )
   EmissiveColor=Desaturation(
    None=MaterialFunctionCall(Hue Shift Percentage=@n10, Texture=@n2.EmissiveColor) {materialFunction=&HueShift}.Result
    Fraction=ScalarParameter() {group=Effect_HueShift, parameterName=HueShift_EmissiveSaturation}
   )
  ) {attributeSetTypes [#1, #2]}
  n12 = Saturate(None=Add(
    A=FunctionInput() {bUsePreviewValueAsDefault=true, inputName="Customization Mask (_CM B)", inputType=FunctionInput_Scalar, sortPriority=24}
    B=ScalarParameter() {group=Effect_HueShift, parameterName=HueShift_Remove_Customization Mask, sortPriority=-10}
   ))
  FunctionOutput(None=StaticSwitchParameter(True=BlendMaterialAttributes(A=@n2.MaterialAttributes, B=@n11, Alpha=@n12), False=@n1) {group=Customization, parameterName=Effect_Hue_Shift})
 ]}
