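# Material function for weapon skin wraps that adds a view-dependent metallic lookup-curve tint to a material's base color.
# Camera-to-normal dot product samples a curve atlas row through MF_Global_CC_Tex, and a selected channel multiplies base color.
# The blend mask comes from a metallic-based amount or topside masks, shaped by primary, accent and optional left/right decal masks.
# Static switches place the effect on one of two outputs, normal or after accent, and offer mask debugging.
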MaterialFunction /Game/Weapons/WeaponSkins/Wraps/MaterialFunctions/MF_Effect_Metal_LUT
graph {nodes [
  n1 = Multiply(
   A=NamedRerouteUsage()
   B=ScalarParameter() {defaultValue=1, group=Effect_Metal_LUT, parameterName="Metal_LUT_Msk_Primary   1=On  0=Off", sortPriority=16}
  )
  n2 = NamedRerouteUsage()
  n3 = ScalarParameter() {group=Effect_Metal_LUT, parameterName="Metal_LUT_Msk_Accent   1=On  0=Off", sortPriority=16}
  Add(A=@n1, B=Multiply(A=@n2, B=@n3))
  Add()
  n4 = NamedRerouteUsage()
  n5 = TextureSample() {samplerType=SAMPLERTYPE_Normal, texture=/Game/Weapons/FORT_Rifles/T_SCAR_N}
  n6 = Constant3Vector() {constant=<rgba 0.4167 0 0 0>}
  n7 = GetMaterialAttributes(None=FunctionInput(Preview=SetMaterialAttributes!(MaterialAttributes, Normal=@n5.RGB, BaseColor=@n6) {attributeSetTypes [#1, #2]}) {inputName=Material, inputType=FunctionInput_MaterialAttributes}) {attributeGetTypes [
    #2
    #3
    #1
    #4
   ]}
  n8 = NamedRerouteDeclaration(None=@n7.MaterialAttributes) {name=Root_Material, nodeColor=<rgba 0.7812 1 0 1>}
  FunctionOutput(None=StaticSwitchParameter(True=@n4, False=@n8) {group=Customization, parameterName=Effect_MetalLUT After Accent}) {outputName=Result_1}
  FunctionOutput(None=StaticSwitchParameter(True=@n8, False=StaticSwitchParameter(True=@n4, False=@n8) {group=Customization, parameterName=Effect_MetalLUT}) {group=Customization, parameterName=Effect_MetalLUT After Accent})
  n9 = Clamp(None=DotProduct(A=CameraVectorWS(), B=PixelNormalWS()))
  n10 = CurveAtlasRowParameter(CurveTime=@n9) {
   atlas=/Game/Weapons/WeaponSkins/Wraps/Textures/CruveAtlas_Wraps_01
   curve=/Game/Weapons/WeaponSkins/Wraps/Textures/0_ColorCurves/CC_MetalicLUT
   defaultValue=76
   group=Effect_Metal_LUT
   parameterName=Metal_LUT_CC
   primitiveDataIndex=1
   sortPriority=-5
  }
  n11 = ChannelMaskParameter(None=MaterialFunctionCall!(
    Index=ScalarParameter() {defaultValue=76, group=Effect_Metal_LUT, parameterName="Metal_LUT_CC  Index", sortPriority=-5}
    CCinput=@n10
    CCinput_Alpha=@n10
    UV=@n9
   ) {materialFunction=&/MF_Global_CC_Tex}.4v) {group=Effect_Metal_LUT, parameterName=Metal_LUT_Channel, sortPriority=-5}
  Multiply(A=@n11.R)
  n12 = GetMaterialAttributes(None=NamedRerouteUsage()) {attributeGetTypes [#1, #2]}
  n13 = FunctionInput() {inputName=TS_Mask_02, inputType=FunctionInput_Vector4, sortPriority=50}
  n14 = FunctionInput() {inputName=TS_Mask_01, inputType=FunctionInput_Vector4, sortPriority=50}
  n15 = ScalarParameter() {group=Effect_Metal_LUT, parameterName=Metal_Use_Metalic Low, sortPriority=0}
  n16 = ScalarParameter() {defaultValue=1, group=Effect_Metal_LUT, parameterName=Metal_Use_Metalic High, sortPriority=0}
  n17 = NamedRerouteUsage()
  n18 = Add(
   A=Saturate(None=LinearInterpolate(A=@n15, B=@n16, Alpha=@n17))
   B=ScalarParameter() {defaultValue=1, group=Effect_Metal_LUT, parameterName="Metal_Use_Metalic Channel  0=On", sortPriority=0}
  )
  n19 = StaticSwitchParameter(
   True=StaticSwitchParameter(
    True=ChannelMaskParameter(None=@n13) {group=Effect_Metal_LUT, parameterName=Metal_LUT_TS_Mask_02_Channels, sortPriority=10}.R
    False=ChannelMaskParameter(None=@n14) {group=Effect_Metal_LUT, parameterName=Metal_LUT_TS_Mask_01_Channels, sortPriority=10}.R
   ) {group=Effect_Metal_LUT, parameterName="Metal_LUT_TS_Mask_01-------->TS_Mask_02", sortPriority=1}
   False=Multiply(A=ScalarParameter() {defaultValue=1, group=Effect_Metal_LUT, parameterName=Metal_LUT_Amount_Simple, sortPriority=0}, B=Saturate(None=@n18))
  ) {desc="Only works with Topside masks. ", group=Effect_Metal_LUT, parameterName="Metal_LUT_Simple-------->TS_Mask_01", sortPriority=1}
  n20 = Add(A=Multiply(A=@n19, B=Saturate(None=@n1)), B=Multiply(A=@n2, B=@n3))
  n21 = FunctionInput(Preview=Constant()) {bUsePreviewValueAsDefault=true, inputName=Left_Right_Mask, inputType=FunctionInput_Scalar, sortPriority=100}
  n22 = LinearInterpolate(
   A=Multiply(A=@n20, B=OneMinus(None=@n21))
   B=Add(A=@n20, B=@n21)
   Alpha=ScalarParameter() {group=Effect_Metal_LUT, parameterName="Metal_LUT_Msk_Left/Right_Decal 1=On  0=Off"}
  )
  n23 = Multiply(
   A=Saturate(None=StaticSwitchParameter(True=@n22, False=@n20) {group=Customization, parameterName=Function_Left_Right_Decal})
   B=ScalarParameter() {defaultValue=1, group=Effect_Metal_LUT, parameterName=Metal_LUT_Metalic_Strength, sortPriority=13}
  ) {constB=3}
  n24 = LinearInterpolate(A=@n12.BaseColor, B=Multiply(A=@n12.BaseColor, B=@n11.R), Alpha=@n23)
  NamedRerouteDeclaration(None=SetMaterialAttributes!(
    MaterialAttributes=@n12.MaterialAttributes
    BaseColor=StaticSwitchParameter(True=@n23, False=@n24) {group=Effect_Metal_LUT, parameterName=Metal_LUT_MaskDebuging, sortPriority=1}
   ) {attributeSetTypes [#2]}) {name=Base_Material, nodeColor=<rgba 1 0.0234 0 1>}
  n25 = TextureSample() {samplerType=SAMPLERTYPE_Masks, texture=/Game/Weapons/FORT_Rifles/Textures/T_Scar_CM}
  n26 = FunctionInput(Preview=@n25.G) {bUsePreviewValueAsDefault=true, inputName=Mask Crevise, inputType=FunctionInput_Scalar, sortPriority=24, previewValue=<1 0 0 1>}
  NamedRerouteDeclaration(None=@n26) {name=Msk_Curvature, nodeColor=<rgba 0 1 0.7422 1>}
  n27 = FunctionInput(Preview=@n25.R) {bUsePreviewValueAsDefault=true, inputName=Mask Primary, inputType=FunctionInput_Scalar, sortPriority=25, previewValue=<1 0 0 1>}
  NamedRerouteDeclaration(None=@n27) {name=Msk_Primary, nodeColor=<rgba 0 1 0.7656 1>}
  n28 = FunctionInput(Preview=@n25.B) {bUsePreviewValueAsDefault=true, inputName=Mask Secondary, inputType=FunctionInput_Scalar, sortPriority=26, previewValue=<1 0 0 1>}
  NamedRerouteDeclaration(None=@n28) {name=Msk_Accent, nodeColor=<rgba 0 1 0.7656 1>}
  NamedRerouteDeclaration(None=@n7.Metallic) {name=Root_metalic, nodeColor=<rgba 0.5234 1 0 1>}
  NamedRerouteDeclaration(None=@n7.Normal) {name=Root_Normal, nodeColor=<rgba 0.4062 0 1 1>}
  OneMinus(None=Fresnel() {baseReflectFraction=0, exponent=1})
  Saturate(None=@n23)
  ScalarParameter() {group=Effect_Metal_LUT, parameterName=Metal_LUT_Msk_Accent Fill White, sortPriority=15}
 ]}
