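# Weapon wrap material function that applies a multicolor gradient effect to an incoming material, toggled by the Effect_MultiColor static switch.
# Gradient coordinates combine a pre-skinned position axis, masked Fresnel and distortion, feeding MF_MulticolorGrad.
# The gradient lerps into base color and emissive, weighted by simple or TS channel masks, Fresnel and down-facing normal dampening.
# Primary and secondary masks drive the final attribute blend, alongside an optional accent material.
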MaterialFunction /Game/Weapons/WeaponSkins/Wraps/MaterialFunctions/MF_Effect_Multicolor
graph {nodes [
  n1 = NamedRerouteUsage()
  n2 = NamedRerouteUsage()
  n3 = TextureSample() {samplerType=SAMPLERTYPE_Masks, texture=/Game/Weapons/FORT_Rifles/Textures/T_Scar_CM}
  n4 = FunctionInput(Preview=@n3.R) {bUsePreviewValueAsDefault=true, inputName=Mask Primary, inputType=FunctionInput_Scalar, sortPriority=24, previewValue=<1 0 0 1>}
  n5 = FunctionInput(Preview=@n3.B) {bUsePreviewValueAsDefault=true, inputName=Mask Secondary, inputType=FunctionInput_Scalar, sortPriority=24, previewValue=<1 0 0 1>}
  n6 = Add(
   A=@n4
   B=Multiply(
    A=NamedRerouteDeclaration(None=@n5) {name=Accent_Mask, nodeColor=<rgba 0 1 0.0156 1>}
    B=ScalarParameter() {defaultValue=1, group=Effect_Multicolor, parameterName="EMC_Mask_Accent  0=off"}
   ) {constB=0}
  )
  n7 = ScalarParameter() {group=Effect_Multicolor, parameterName="EMC_Mask_Disable 1=0n"}
  n8 = Saturate(None=Saturate(None=Add(A=@n6, B=@n7) {constB=0}))
  n9 = TextureSample() {samplerType=SAMPLERTYPE_Normal, texture=/Game/Weapons/FORT_Rifles/T_SCAR_N}
  n10 = Constant3Vector() {constant=<rgba 1 0 0 0>}
  n11 = GetMaterialAttributes(None=FunctionInput(Preview=SetMaterialAttributes!(MaterialAttributes, Normal=@n9.RGB, BaseColor=@n10) {attributeSetTypes [#1, #2]}) {inputName=Material, inputType=FunctionInput_MaterialAttributes}) {attributeGetTypes [
    #2
    #3
    #1
    #4
   ]}
  FunctionOutput(None=StaticSwitchParameter(
    True=BlendMaterialAttributes(A=@n1, B=@n2, Alpha=@n8) {vertexAttributeBlendType=UseA}
    False=NamedRerouteDeclaration(None=@n11.MaterialAttributes) {name=Root_Material, nodeColor=<rgba 0.7812 1 0 1>}
   ) {group=Customization, parameterName=Effect_MultiColor})
  n12 = SetMaterialAttributes!(
   MaterialAttributes=NamedRerouteUsage()
   BaseColor=VectorParameter() {group=Effect_Accent, parameterName=EMC_Accent_BaseColor, defaultValue=<rgba 0 1 0.132 1>}.RGB
   Roughness=ScalarParameter() {defaultValue=0.3, group=Effect_Accent, parameterName=EMC_Accent_Roughness}
   Specular=ScalarParameter() {defaultValue=0.5, group=Effect_Accent, parameterName=EMC_Accent_Spec}
   Metallic=ScalarParameter() {defaultValue=0.5, group=Effect_Accent, parameterName=EMC_Accent_Metalic}
   Normal
   EmissiveColor
  ) {attributeSetTypes [
    #2
    #3
    #5
    #4
    #1
    #6
   ]}
  NamedRerouteDeclaration(None=@n12) {name=Mat_Accent, nodeColor=<rgba 0 0.3906 1 1>}
  n13 = FunctionInput(Preview=@n3.G) {bUsePreviewValueAsDefault=true, inputName=Mask Crevise, inputType=FunctionInput_Scalar, sortPriority=24, previewValue=<1 0 0 1>}
  NamedRerouteDeclaration(None=@n13) {name=Msk_Curvature, nodeColor=<rgba 0 1 0.7422 1>}
  n14 = Multiply(A=NamedRerouteUsage(), B=ScalarParameter() {group=Effect_Accent, parameterName="EMC_Accent  1=On"})
  n15 = GetMaterialAttributes(None=BlendMaterialAttributes(A=NamedRerouteUsage(), B=NamedRerouteUsage(), Alpha=Saturate(None=@n14)) {vertexAttributeBlendType=UseA}) {attributeGetTypes [#2, #6]}
  n16 = TextureObjectParameter() {
   group=Effect_Multicolor
   parameterName=EMC_GradientTex
   sortPriority=0
   texture=/Game/Characters/Player/Female/Medium/Bodies/F_Med_Soldier_01/Skins/F_MED_COM_Taxi_Upgraded/Skins/Multi/T_multicolor_gradient_v2_D
  }
  n17 = FunctionInput() {inputName=TS_Mask_02, inputType=FunctionInput_Vector4, sortPriority=50}
  n18 = FunctionInput() {inputName=TS_Mask_01, inputType=FunctionInput_Vector4, sortPriority=49}
  n19 = StaticSwitchParameter(
   True=StaticSwitchParameter(
    True=ChannelMaskParameter(None=@n17) {group=Effect_Multicolor_Masking, parameterName=Irad_TS_Mask_02_Channels, sortPriority=10}.R
    False=ChannelMaskParameter(None=@n18) {group=Effect_Multicolor_Masking, parameterName=Irad_TS_Mask_01_Channels, sortPriority=10}.R
   ) {group=Effect_Multicolor_Masking, parameterName="EMC_Multi_TS_Mask_01-------->TS_Mask_02", sortPriority=1}
   False=ScalarParameter() {defaultValue=1, group=Effect_Multicolor_Masking, parameterName=EMC_Mask_Amount_Simple, sortPriority=0}
  ) {desc="Only works with Topside masks. ", group=Effect_Multicolor_Masking, parameterName="EMC_Mask_Simple-------->TS_Mask_01", sortPriority=1}
  n20 = Fresnel(ExponentIn=ScalarParameter() {defaultValue=1, group=Effect_Multicolor, parameterName=EMC_Fresnel, sortPriority=10}) {baseReflectFraction=0, exponent=1}
  n21 = ScalarParameter() {defaultValue=1, group=Effect_Multicolor, parameterName=EMC_FresnelStrength, sortPriority=11}
  n22 = Multiply(A=@n19, B=Saturate(None=Multiply(A=@n20, B=@n21) {constB=2})) {constB=2}
  n23 = Multiply(A=@n22, B=ScalarParameter() {defaultValue=1, group=Effect_Multicolor, parameterName=EMC_UV_ViewAngle, sortPriority=22})
  n24 = VertexInterpolator(VS=PreSkinnedPosition())
  n25 = PreSkinnedLocalBounds()
  n26 = ChannelMaskParameter(None=LinearInterpolate(
    A=Divide(A=@n24.PS, B=@n25.Max)
    B=NamedRerouteUsage()
    Alpha=ScalarParameter() {group=Effect_Multicolor, parameterName="EMC_UV_USE_Bounds 1=ON", sortPriority=20}
   )) {group=Effect_Multicolor, maskChannel=Blue, parameterName=EMC_UV_Axis, sortPriority=19, defaultValue=<rgba 0 0 1 0>}
  n27 = Add(
   A=Multiply(A=@n26.R, B=ScalarParameter() {defaultValue=1, group=Effect_Multicolor, parameterName=EMC_UV_Tiling, sortPriority=20})
   B=ScalarParameter() {group=Effect_Multicolor, parameterName=EMC_UV_Offset, sortPriority=21}
  )
  n28 = Add(A=NamedRerouteUsage()) {constB=-0.5}
  n29 = ScalarParameter() {group=Effect_Multicolor_Masking, parameterName=EMC_Distortion_Curavature_Amount, sortPriority=0}
  n30 = MaterialFunctionCall!(
   Phase Shift=ScalarParameter() {defaultValue=1, group=Effect_Multicolor, parameterName=EMC_GradientShift, sortPriority=41}
   UV=Add(A=Add(A=@n23, B=@n27), B=Multiply(A=@n28, B=@n29))
   height texture=NamedRerouteUsage()
   Gradient=@n16
   Phases=ScalarParameter() {defaultValue=4, group=Effect_Multicolor, parameterName=EMC_GradientPhases, sortPriority=40}
  ) {materialFunction=/Game/Characters/Player/Female/Medium/Bodies/F_Med_Soldier_01/Skins/F_MED_COM_Taxi_Upgraded/Skins/Multi/MF_MulticolorGrad}
  n31 = Add(A=@n30.Result, B=ScalarParameter() {group=Effect_Multicolor, parameterName=EMC_BaseColorBrighten, sortPriority=31})
  n32 = PreSkinnedNormal()
  n33 = Multiply(
   A=DotProduct(A=VertexInterpolator(VS=@n32).PS, B=Constant3Vector() {constant=<rgba 0 0 -1 0>})
   B=ScalarParameter() {group=Effect_Multicolor, parameterName=EMC_DownNormalDampen, sortPriority=100}
  )
  n34 = Multiply(A=OneMinus(None=Saturate(None=@n33)), B=@n22)
  n35 = Multiply(A=@n34, B=ScalarParameter() {defaultValue=1, group=Effect_Multicolor, parameterName=EMC_BaseColorStrength})
  n36 = ScalarParameter() {defaultValue=1, group=Effect_Multicolor, parameterName=EMC_EmissiveStrength}
  n37 = SetMaterialAttributes(
   MaterialAttributes=NamedRerouteUsage()
   BaseColor=Saturate(None=LinearInterpolate(A=@n15.BaseColor, B=@n31, Alpha=@n35))
   EmissiveColor=LinearInterpolate(A=@n15.EmissiveColor, B=@n30.Result, Alpha=Multiply(A=@n34, B=@n36))
  ) {attributeSetTypes [
    #2
    #3
    #5
    #4
    #1
    #6
   ]}
  NamedRerouteDeclaration(None=@n37) {name=Mat_Base, nodeColor=<rgba 0 0.3906 1 1>}
  NamedRerouteDeclaration(None=FunctionInput(Preview=Constant4Vector().RGBA) {bUsePreviewValueAsDefault=true, inputName=IsState, inputType=FunctionInput_Vector4, sortPriority=4}) {name=IsState, nodeColor=<rgba 0.1497 1 0.1633 1>}
  NamedRerouteDeclaration(None=FunctionInput(Preview=Constant2Vector()) {bUsePreviewValueAsDefault=true, inputName=IsState_B, inputType=FunctionInput_Vector2, sortPriority=4}) {name=IsState_B, nodeColor=<rgba 0.1497 1 0.1633 1>}
  n38 = PreSkinnedLocalBounds()
  n39 = Subtract(A=VertexInterpolator(VS=PreSkinnedPosition()).PS, B=@n38.Min)
  n40 = MaterialFunctionCall(Color=Subtract(A=@n38.Max, B=@n38.Max)) {materialFunction=&SplitComponents}
  n41 = Max(A=Max(A=@n40.R, B=@n40.R), B=@n40.B)
  NamedRerouteDeclaration(None=FunctionInput(Preview=Divide(A=@n39, B=@n41)) {bUsePreviewValueAsDefault=true, inputName=Calcated_Bounds, sortPriority=11}) {name=Caclcated_Bounds, nodeColor=<rgba 0 0.9297 1 1>}
  n42 = StaticSwitchParameter(
   True=StaticSwitchParameter(
    True=ChannelMaskParameter(None=@n17) {group=Effect_Multicolor_Masking, parameterName=Irad_TS_Mask_02_Channels, sortPriority=10}.R
    False=ChannelMaskParameter(None=@n18) {group=Effect_Multicolor_Masking, parameterName=Irad_TS_Mask_01_Channels, sortPriority=10}.R
   ) {group=Effect_Multicolor_Masking, parameterName="EMC_Distortion_TS_Mask_01-------->TS_Mask_02", sortPriority=1}
   False=ScalarParameter() {defaultValue=1, group=Effect_Multicolor_Masking, parameterName=EMC_Distortion_Amount_Simple, sortPriority=0}
  ) {
   desc="Only works with Topside masks. "
   group=Effect_Multicolor_Masking
   parameterName="EMC_Distortion_Simple-------->TS_Mask_01"
   sortPriority=1
  }
  NamedRerouteDeclaration(None=@n42) {name=Distortion, nodeColor=<rgba 0.6875 1 0 1>}
  n43 = PreSkinnedNormal()
  NamedRerouteDeclaration(None=FunctionInput(Preview=VertexInterpolator(VS=@n43).PS) {bUsePreviewValueAsDefault=true, inputName=Pre-Skinned Local Normal, sortPriority=10}) {name=PreSkin_Normal, nodeColor=<rgba 0.8281 1 0 1>}
  NamedRerouteDeclaration(None=@n11.Normal) {name=Root_Normal, nodeColor=<rgba 0.4062 0 1 1>}
 ]}
