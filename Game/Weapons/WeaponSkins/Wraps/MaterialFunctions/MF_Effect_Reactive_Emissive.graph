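# Material function for weapon wrap skins that makes emissive react to a timed trigger, fading between start and end brightness.
# Elapsed time since a timestamp parameter, apparently set from Blueprint, is normalized by a duration and drives the brightness lerp.
# That multiplies incoming emissive and a customization mask with additive override, zeroing emissive outside the mask.
# A static switch either outputs the modified attributes or passes the input through.
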MaterialFunction /Game/Weapons/WeaponSkins/Wraps/MaterialFunctions/MF_Effect_Reactive_Emissive
graph {nodes [
  n1 = FunctionInput() {inputType=FunctionInput_MaterialAttributes}
  n2 = GetMaterialAttributes(None=@n1) {attributeGetTypes [#1]}
  n3 = FunctionInput() {inputName="Customization Mask (_CM B)", inputType=FunctionInput_Scalar, sortPriority=24, previewValue=<1 0 0 1>}
  n4 = ScalarParameter() {group=Effect_Reactive_Emissive, parameterName=Rect_Emissive_Custmization_Mask, sortPriority=-10}
  n5 = ScalarParameter() {defaultValue=1, group=Effect_Reactive_Emissive, parameterName=Effect_Brightnees_Start_Value}
  n6 = ScalarParameter() {group=Effect_Reactive_Emissive, parameterName=Effect_Brightnees_End_Value}
  n7 = Time()
  n8 = ScalarParameter() {defaultValue=-9999999, group=Effect_Reactive_Emissive, parameterName=TimeFrom-BP----------------}
  n9 = Saturate(None=Divide(A=Subtract(A=@n7, B=@n8), B=ScalarParameter() {defaultValue=0.25, group=Effect_Reactive_Emissive, parameterName=Effect_Brightnees_Duration}))
  n10 = Multiply(A=Saturate(None=Add(A=@n3, B=@n4)), B=Multiply(A=@n2.EmissiveColor, B=LinearInterpolate(A=@n5, B=@n6, Alpha=@n9)))
  FunctionOutput(None=StaticSwitchParameter(
    True=SetMaterialAttributes!(MaterialAttributes=@n2.MaterialAttributes, EmissiveColor=@n10) {attributeSetTypes [#1]}
    False=@n1
   ) {group=Customization, parameterName=Effect_Reactive_Emissive})
 ]}
