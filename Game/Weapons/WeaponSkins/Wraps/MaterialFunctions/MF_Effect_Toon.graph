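# Weapon-wrap material function that applies a switchable toon effect to incoming material attributes.
# Sun-versus-normal and camera-versus-normal terms are remapped through curve-atlas ramps, giving banded shading and optional fake specular/roughness highlights.
# These tint customization-masked base colors, then Fresnel and curvature darken outlines.
# The result drives emissive and optionally base color, a static PBR toggle controls roughness, specular and metallic, and mask A can blend back the original material.
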MaterialFunction /Game/Weapons/WeaponSkins/Wraps/MaterialFunctions/MF_Effect_Toon
graph {nodes [
  n1 = NamedRerouteUsage()
  n2 = ScalarParameter() {defaultValue=37, group=Effect_Toon, parameterName="EF_Toon_CC_Remap  Index"}
  n3 = CollectionParameter() {
   collection=&FortniteMaterialParameters
   parameterName=SunlightDirectionalVector
  }
  n4 = PixelNormalWS()
  n5 = ConstantBiasScale(None=DotProduct(A=ComponentMask(None=@n3) {b=true, g=true, r=true}, B=@n4))
  n6 = CurveAtlasRowParameter(CurveTime=@n5) {
   atlas=/Game/Weapons/WeaponSkins/Wraps/Textures/CruveAtlas_Wraps_01
   curve=/Game/Weapons/WeaponSkins/Wraps/Textures/C_Toon
   defaultValue=37
   group=Effect_Toon
   parameterName=EF_Toon_CC_Remap
  }
  n7 = MaterialFunctionCall!(Index=@n2, CCinput=@n6, CCinput_Alpha=@n6, UV=@n5) {materialFunction=&/MF_Global_CC_Tex}
  n8 = ComponentMask(None=@n7.3V) {r=true}
  n9 = FunctionInput() {inputName=Material, inputType=FunctionInput_MaterialAttributes}
  n10 = GetMaterialAttributes(None=@n9) {attributeGetTypes [
    #1
    #2
    #3
    #4
    #5
    #6
   ]}
  n11 = VectorParameter() {group=Effect_Toon, parameterName=EF_Toon_SimpleSecondaryColor, sortPriority=-5, defaultValue=<rgba 0.0142 0.055 0.1354 1>}
  n12 = VectorParameter() {group=Effect_Toon, parameterName=EF_Toon_SimplePrimaryColor, sortPriority=-5, defaultValue=<rgba 0.4427 0 0 1>}
  n13 = FunctionInput() {bUsePreviewValueAsDefault=true, inputName=Customization Mask A, inputType=FunctionInput_Scalar, sortPriority=24}
  n14 = LinearInterpolate(A=@n11.RGB, B=@n12.RGB, Alpha=@n13)
  n15 = GetMaterialAttributes(None=@n9) {attributeGetTypes [
    #1
    #2
    #3
    #4
    #5
    #6
   ]}
  n16 = ScalarParameter() {defaultValue=1, group=Effect_Toon, parameterName="EF_Toon_Simple_Secondary_Color_with_BaseColor  1=Disable 0=Active", sortPriority=-5}
  n17 = LinearInterpolate(A=@n11.RGB, B=@n15.BaseColor, Alpha=Saturate(None=Add(A=@n16, B=@n13)))
  n18 = ScalarParameter() {group=Effect_Toon, parameterName=EF_Toon_Customization Mask Most of the weapon, sortPriority=-10}
  n19 = FunctionInput() {bUsePreviewValueAsDefault=true, inputName=Customization Mask B, inputType=FunctionInput_Scalar, sortPriority=24}
  n20 = Add(
   A=Add(A=@n13, B=Multiply(A=@n19, B=@n18))
   B=ScalarParameter() {group=Effect_Toon, parameterName=EF_Toon_Remove_Customization Mask, sortPriority=-10}
  )
  n21 = Add(
   A=LinearInterpolate(
    A=@n10.BaseColor
    B=StaticSwitchParameter(True=@n14, False=@n17) {group=Effect_Toon, parameterName=EF_Toon_Simple_BaseColor, sortPriority=-5}
    Alpha=Saturate(None=@n20)
   )
   B=Multiply(
    A=@n15.EmissiveColor
    B=ScalarParameter() {group=Effect_Toon, parameterName="EF_Toon_Add Emissive from EmissiveChannel 0=off 1=active", sortPriority=1}
   )
  )
  n22 = Clamp(None=DotProduct(A=CameraVectorWS(), B=@n4))
  n23 = ScalarParameter() {defaultValue=1, group=Effect_Toon, parameterName=EF_Toon_FakeSpecularStrength_Min}
  n24 = ScalarParameter() {defaultValue=1, group=Effect_Toon, parameterName=EF_Toon_FakeSpecularStrength_Max}
  n25 = ScalarParameter() {group=Effect_Toon, parameterName="EF_Toon_FakeRoughnessandSpec 0=off 1=active"}
  n26 = LinearInterpolate(A=@n22, B=Multiply(A=@n22, B=LinearInterpolate(A=@n23, B=@n24, Alpha=@n15.Specular)), Alpha=@n25)
  n27 = CurveAtlasRowParameter(CurveTime=@n26) {
   atlas=&CruveAtlas_Wraps_01
   curve=&C_Toon
   defaultValue=37
   group=Effect_Toon
   parameterName=EF_Toon_CC_Remap
  }
  n28 = MaterialFunctionCall(Index=@n2, CCinput=@n27, UV=@n26) {materialFunction=&MF_Global_CC_Tex}
  n29 = Multiply(A=Multiply(A=@n7.Alpha, B=ComponentMask(None=@n28.3V) {b=true}), B=@n21)
  n30 = LinearInterpolate(
   A=ScalarParameter() {defaultValue=15, group=Effect_Toon, parameterName=EF_Toon_FakeroughnessStrength_Max}
   B=ScalarParameter() {group=Effect_Toon, parameterName=EF_Toon_FakeRoughnessStrength_Min}
   Alpha=@n15.Roughness
  )
  n31 = Add(A=Multiply(A=@n21, B=@n8), B=LinearInterpolate(B=Multiply(A=@n29, B=@n30), Alpha=@n25))
  n32 = ScalarParameter() {group=Effect_Toon, parameterName="EF-Toon-Outline-Disable 0=off 1=on -------------------------------------", sortPriority=49}
  n33 = ScalarParameter() {defaultValue=10, group=Effect_Toon, parameterName=EF_Toon_Curvature_Mask A, sortPriority=50}
  n34 = ScalarParameter() {defaultValue=-7, group=Effect_Toon, parameterName=EF_Toon_Curvature_Mask B, sortPriority=50}
  n35 = FunctionInput(Preview=TextureSample() {samplerType=SAMPLERTYPE_Masks, texture=/Game/Weapons/FORT_Rifles/Textures/T_Scar_CM}.G) {bUsePreviewValueAsDefault=true, inputName=Curavture, inputType=FunctionInput_Scalar, sortPriority=24}
  n36 = Fresnel(ExponentIn=ScalarParameter() {defaultValue=4.7243, group=Effect_Toon, parameterName=EF_Toon_OutlineExponent1, sortPriority=100}) {baseReflectFraction=0}
  n37 = LinearInterpolate(
   A=ScalarParameter() {defaultValue=-5.4358, group=Effect_Toon, parameterName=EF_Toon_OutlineC1, sortPriority=100}
   B=ScalarParameter() {defaultValue=41.451, group=Effect_Toon, parameterName=EF_Toon_OutlineC2, sortPriority=100}
   Alpha=Power(
    Base=OneMinus(None=@n36)
    Exp=ScalarParameter() {defaultValue=4.6385, group=Effect_Toon, parameterName=EF_Toon_OutlineExponent2, sortPriority=100}
   )
  )
  n38 = Saturate(None=Multiply(A=LinearInterpolate(A=@n33, B=@n34, Alpha=@n35), B=@n37))
  n39 = Multiply(A=@n31, B=Saturate(None=Add(A=@n32, B=@n38)))
  n40 = ScalarParameter() {defaultValue=1, group=Effect_Toon, parameterName=EF_Toon_EmissiveStrength, sortPriority=1}
  n41 = ScalarParameter() {defaultValue=1, desc="PBR-Effect  needs to be active", group=Effect_Toon, parameterName=EF_Toon_RoughnessMax, sortPriority=10}
  n42 = LinearInterpolate(
   A=ScalarParameter() {desc="PBR-Effect  needs to be active", group=Effect_Toon, parameterName=EF_Toon_RoughnessMin, sortPriority=10}
   B=@n41
   Alpha=@n15.Roughness
  )
  n43 = StaticBoolParameter() {group=Effect_Toon, parameterName=EF-Toon-PBR-Effect--, sortPriority=5}
  n44 = ScalarParameter() {defaultValue=1, desc="PBR-Effect  needs to be active", group=Effect_Toon, parameterName=EF_Toon_SpecularMax, sortPriority=10}
  n45 = LinearInterpolate(
   A=ScalarParameter() {desc="PBR-Effect  needs to be active", group=Effect_Toon, parameterName=EF_Toon_SpecularMin, sortPriority=10}
   B=@n44
   Alpha=@n15.Specular
  )
  n46 = Constant()
  n47 = SetMaterialAttributes!(
   MaterialAttributes=@n15.MaterialAttributes
   EmissiveColor=Multiply(A=@n39, B=Multiply(A=@n8, B=@n40)) {constB=2}
   BaseColor=LinearInterpolate(B=@n39, Alpha=ScalarParameter() {group=Effect_Toon, parameterName="EF_Toon_BaseColor 0=black 1=basecolor", sortPriority=0})
   Roughness=StaticSwitch(True=@n42, False=Constant() {r=1}, Value=@n43)
   Specular=StaticSwitch(True=@n45, False=@n46, Value=@n43)
   Metallic=StaticSwitch(True=@n15.Metallic, False=@n46, Value=@n43)
   Normal=@n15.Normal
  ) {attributeSetTypes [
    #6
    #1
    #3
    #4
    #5
    #2
   ]}
  n48 = Saturate(None=Add(A=@n13, B=ScalarParameter() {defaultValue=1, group=Effect_Toon, parameterName="EF_Toon_Primary_Mask_BC  0=Active  1=Off", sortPriority=0}))
  FunctionOutput(None=StaticSwitchParameter(True=BlendMaterialAttributes(A=@n1, B=@n47, Alpha=@n48), False=@n9) {group=Customization, parameterName=Effect_Toon})
  n49 = ScalarParameter() {group=Effect_Toon, parameterName="EF-Toon-PBR-Effect--0=off---1=active-----------", sortPriority=5}
  LinearInterpolate(B=@n45, Alpha=@n49)
  LinearInterpolate(B=@n42, Alpha=@n49) {constA=1}
  LinearInterpolate(B=@n15.Metallic, Alpha=@n49)
  NamedRerouteDeclaration(None=@n10.MaterialAttributes) {name=Root_Material, nodeColor=<rgba 1 0 0.0469 1>}
  Saturate(None=@n31)
  Saturate(None=@n37)
  Transform(None=@n15.Normal)
 ]}
