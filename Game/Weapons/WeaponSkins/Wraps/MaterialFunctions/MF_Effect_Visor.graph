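# Wrap material function that layers an animated, glowing visor effect onto weapons, vehicles, characters, mechs, backpacks and pickaxes, behind a static switch.
# A flipbook mask crossfades between frames on side-projected UVs, with scale, rotation, offset and depth chosen per item type.
# Two optional camera-shifted layers fake depth; mobile uses only the surface layer.
# Fresnel and edge glow shape the tinted emissive, which can darken base color and blends in through input masks.
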MaterialFunction /Game/Weapons/WeaponSkins/Wraps/MaterialFunctions/MF_Effect_Visor
graph {
 properties {description=10}
 nodes [
  n1 = GetMaterialAttributes(None=NamedRerouteUsage()) {attributeGetTypes [#1, #2]}
  n2 = VectorParameter() {group=Effect_Visor, parameterName="EVisor_SubImages(V2)", sortPriority=2, defaultValue=<rgba 2 2 0 1>}
  n3 = MaterialFunctionCall(Float2=@n2.RGB) {materialFunction=&BreakOutFloat2Components}
  n4 = Multiply(A=Time(), B=ScalarParameter() {defaultValue=2, group=Effect_Visor, parameterName=EVisor_Frequency, sortPriority=3})
  n5 = Frac(None=@n4)
  n6 = Fmod(A=Subtract(A=@n4, B=@n5), B=Multiply(A=@n3.R, B=@n3.R))
  n7 = AppendVector(A=Fmod(A=@n6, B=@n3.R), B=Floor(None=Divide(A=@n6, B=@n3.R)))
  n8 = NamedRerouteUsage()
  n9 = TextureSample(UVs=Divide(A=Add(A=@n7, B=@n8), B=NamedRerouteUsage()) {constB=3}, Tex=NamedRerouteUsage())
  n10 = Add(A=@n9.A, B=ScalarParameter() {defaultValue=-0.3, group=Effect_Visor, parameterName=EVisor_EdgeGlow_inset, sortPriority=2}) {constB=-0.3}
  n11 = Power(
   Base=OneMinus(None=Saturate(None=@n10))
   Exp=ScalarParameter() {defaultValue=1.5, group=Effect_Visor, parameterName=EVisor_EdgeGlow_Fallow, sortPriority=2}
  ) {constExponent=1.5}
  n12 = ScalarParameter() {defaultValue=0.45, group=Effect_Visor, parameterName=EVisor_EdgeGlow_Brightness, sortPriority=2}
  n13 = ScalarParameter() {defaultValue=1.5, group=Effect_Visor, parameterName=EVisor_Fresnel_Expon, sortPriority=2}
  n14 = Constant()
  n15 = FunctionInput() {inputType=FunctionInput_MaterialAttributes}
  n16 = Transform(None=GetMaterialAttributes(None=@n15) {attributeGetTypes [#3]}.Normal)
  n17 = Power(Base=OneMinus(None=Fresnel(ExponentIn=@n13, BaseReflectFractionIn=@n14, Normal=@n16))) {constExponent=3}
  n18 = MaterialFunctionCall(Float2=@n2.RGB) {materialFunction=&BreakOutFloat2Components}
  n19 = Frac(None=@n4)
  n20 = Fmod(A=Add(A=Subtract(A=@n4, B=@n19)), B=Multiply(A=@n18.R, B=@n18.R))
  n21 = AppendVector(A=Fmod(A=@n20, B=@n18.R), B=Floor(None=Divide(A=@n20, B=@n18.R)))
  n22 = Divide(A=Add(A=NamedRerouteUsage(), B=@n21), B=NamedRerouteUsage()) {constB=3}
  n23 = NamedRerouteUsage()
  n24 = Multiply(
   A=LinearInterpolate(A=@n9.A, B=TextureSample(UVs=@n22, Tex=@n23).A, Alpha=@n5)
   B=TextureSample(UVs=NamedRerouteUsage(), Tex=NamedRerouteUsage()).R
  )
  n25 = ScalarParameter() {defaultValue=-0.48, group=Effect_Visor, parameterName=EVisor_SurfaceDistanceOffset, sortPriority=11}
  n26 = Multiply(A=Saturate(None=Add(A=@n24, B=@n25) {constB=-0.25}), B=Constant() {r=14})
  n27 = NamedRerouteUsage()
  n28 = TextureSample(UVs=NamedRerouteUsage(), Tex=@n27)
  n29 = Constant() {r=1}
  n30 = StaticBoolParameter() {group=Effect_Visor, parameterName=EVisor_Pattern BlueChannle}
  n31 = Multiply(A=Saturate(None=@n26), B=StaticSwitch(True=@n28.B, False=@n29, Value=@n30))
  n32 = Divide(A=Add(A=@n7, B=NamedRerouteUsage()), B=NamedRerouteUsage()) {constB=3}
  n33 = NamedRerouteUsage()
  n34 = Divide(A=Add(A=NamedRerouteUsage(), B=@n21), B=NamedRerouteUsage()) {constB=3}
  n35 = NamedRerouteUsage()
  n36 = Multiply(
   A=TextureSample(UVs=NamedRerouteUsage(), Tex=NamedRerouteUsage()).R
   B=LinearInterpolate(A=TextureSample(UVs=@n32, Tex=@n33).A, B=TextureSample(UVs=@n34, Tex=@n35).A, Alpha=@n5)
  )
  n37 = ScalarParameter() {defaultValue=-0.4, group=Effect_Visor, parameterName=EVisor_Ly1DistanceOffset}
  n38 = Multiply(A=Saturate(None=Add(A=@n36, B=@n37) {constB=-0.25}), B=Constant() {r=14})
  n39 = NamedRerouteUsage()
  n40 = Multiply(
   A=Multiply(A=ScalarParameter() {defaultValue=0.5, group=Effect_Visor, parameterName=EVisor_Ly1Strengh}, B=Saturate(None=@n38))
   B=StaticSwitch(True=TextureSample(UVs=@n39, Tex=@n27).B, False=@n29, Value=@n30)
  )
  n41 = Divide(A=Add(A=@n7, B=NamedRerouteUsage()), B=NamedRerouteUsage()) {constB=3}
  n42 = NamedRerouteUsage()
  n43 = Divide(A=Add(A=NamedRerouteUsage(), B=@n21), B=NamedRerouteUsage()) {constB=3}
  n44 = NamedRerouteUsage()
  n45 = Multiply(
   A=TextureSample(UVs=NamedRerouteUsage(), Tex=NamedRerouteUsage()).R
   B=LinearInterpolate(A=TextureSample(UVs=@n41, Tex=@n42).A, B=TextureSample(UVs=@n43, Tex=@n44).A, Alpha=@n5)
  )
  n46 = ScalarParameter() {defaultValue=-0.6, group=Effect_Visor, parameterName=EVisor_Ly2DistanceOffset}
  n47 = Multiply(A=Saturate(None=Add(A=@n45, B=@n46) {constB=-0.25}), B=Constant() {r=14})
  n48 = NamedRerouteUsage()
  n49 = Multiply(
   A=Multiply(A=Saturate(None=@n47), B=ScalarParameter() {defaultValue=0.3, group=Effect_Visor, parameterName=EVisor_Ly2Strengh})
   B=StaticSwitch(True=TextureSample(UVs=@n48, Tex=@n27).B, False=@n29, Value=@n30)
  )
  n50 = StaticSwitchParameter(True=@n31, False=Add(A=@n31, B=Add(A=@n40, B=@n49))) {group=Effect_Visor_Depth, parameterName=EVisor_Disable_Layers, sortPriority=0}
  n51 = ShadingPathSwitch(Default=@n50, Deferred=@n50, Mobile=@n26) {desc=Migrated from FeatureLevelSwitch}
  n52 = ScalarParameter() {defaultValue=1, group=Effect_Visor, parameterName=EVisor_Emissive mult, sortPriority=5}
  n53 = ScalarParameter() {defaultValue=6, group=Effect_Visor, parameterName=EVisor_emissive mult_MobileMax, sortPriority=7}
  n54 = Multiply(
   A=Multiply(A=Add(A=@n11, B=@n12) {constB=0.45}, B=Multiply(A=@n17, B=@n51))
   B=ShadingPathSwitch(Default=@n52, Deferred=@n52, Mobile=Clamp(None=@n52, Max=@n53)) {desc=Migrated from FeatureLevelSwitch}
  )
  n55 = Saturate(None=@n54)
  n56 = Saturate(None=Add(
    A=ScalarParameter() {defaultValue=1, group=Effect_Visor, parameterName="EVisor_Darken_Base  0=on 1=off", sortPriority=1}
    B=OneMinus(None=@n55)
   ))
  n57 = VectorParameter() {group=Effect_Visor, parameterName=EVisor_Emissive Color, sortPriority=8, defaultValue=<rgba 1 1 1 1>}
  n58 = VectorParameter() {group=Effect_Visor, parameterName=EVisor_Emissive Color_B, sortPriority=8, defaultValue=<rgba 0 0 0 1>}
  n59 = Power(Base=SphereMask(
    A=NamedRerouteUsage()
    B=Constant() {r=0.5}
    Radius=ScalarParameter() {defaultValue=0.5, group=Effect_Visor, parameterName=EVisor_emissive _SPMSK_Radius}
    Hardness=ScalarParameter() {group=Effect_Visor, parameterName=EVisor_emissive _SPMSK_Hardness}
   ))
  n60 = Multiply(
   A=@n54
   B=LinearInterpolate(
    A=@n57.RGB
    B=LinearInterpolate(A=@n58.RGB, B=@n57.RGB, Alpha=@n59)
    Alpha=ScalarParameter() {group=Effect_Visor, parameterName="EVisor_emissive _SPMSK 0=off 1=0n"}
   )
  )
  n61 = ScalarParameter() {group=Effect_Visor, parameterName="EVisor_PrimaryMsk_Disable  0=off 1=active", sortPriority=-1}
  n62 = TextureSample() {samplerType=SAMPLERTYPE_Masks, texture=/Game/Weapons/FORT_Rifles/Textures/T_Scar_CM}
  n63 = FunctionInput(Preview=@n62.R) {bUsePreviewValueAsDefault=true, inputName=Mask Primary, inputType=FunctionInput_Scalar, sortPriority=24, previewValue=<1 0 0 1>}
  n64 = FunctionInput(Preview=@n62.B) {bUsePreviewValueAsDefault=true, inputName=Mask Secondary, inputType=FunctionInput_Scalar, sortPriority=24, previewValue=<1 0 0 1>}
  n65 = Multiply(
   A=@n64
   B=ScalarParameter() {defaultValue=1, group=Effect_Visor, parameterName="EVisor_SecondaryMSk_Disable  1=off 0=active", sortPriority=-1}
  )
  n66 = Add(A=@n63, B=Saturate(None=@n65))
  n67 = BlendMaterialAttributes(
   A=NamedRerouteUsage()
   B=SetMaterialAttributes!(MaterialAttributes=@n1.MaterialAttributes, BaseColor=Multiply(A=@n1.BaseColor, B=@n56), EmissiveColor=Add(A=@n1.EmissiveColor, B=@n60)) {attributeSetTypes [#1, #2]}
   Alpha=Saturate(None=Add(A=@n61, B=@n66))
  ) {vertexAttributeBlendType=UseA}
  BlendMaterialAttributes(A=@n67, Alpha=@n65) {vertexAttributeBlendType=UseA}
  Constant4Vector()
  FunctionOutput(None=StaticSwitchParameter(True=@n67, False=NamedRerouteUsage()) {desc=TopSide Projection Base - Emissive pan Mask, group=Customization, parameterName=Effect_Visor})
  LinearInterpolate(B=@n28.B, Alpha=ScalarParameter() {group=Effect_Visor, parameterName="EVisor_Pattern BlueChannle 0=off 1=on"}) {constA=1}
  Multiply(A=@n9.A)
  NamedRerouteDeclaration(None=@n15) {name=Root_material, nodeColor=<rgba 0 1 0.3438 1>}
  n68 = ScalarParameter() {defaultValue=1, group=Effect_Visor_UVS, parameterName=EVisor_Side_Scale_Vehicle, sortPriority=150}
  n69 = ScalarParameter() {defaultValue=1, group=Effect_Visor_UVS, parameterName=EVisor_Side_Scale_Vehicle_Side, sortPriority=150}
  n70 = MaterialFunctionCall(Float4=NamedRerouteUsage()) {materialFunction=&BreakOutFloat4Components}
  n71 = LinearInterpolate(
   A=LinearInterpolate(
    A=ScalarParameter() {defaultValue=0.5, group=Effect_Visor_UVS, parameterName=EVisor_Side_Scale_Weapon, sortPriority=150}
    B=LinearInterpolate(A=@n68, B=@n69, Alpha=@n70.G)
    Alpha=@n70.R
   )
   B=LinearInterpolate(
    A=ScalarParameter() {defaultValue=1, group=Effect_Visor_UVS, parameterName=EVisor_Side_Scale_Mech, sortPriority=150}
    B=ScalarParameter() {defaultValue=1, group=Effect_Visor_UVS, parameterName=EVisor_Side_Scale_Character, sortPriority=150}
    Alpha=@n70.A
   )
   Alpha=@n70.B
  )
  n72 = ScalarParameter() {defaultValue=1, group=Effect_Visor_UVS, parameterName=EVisor_Side_Scale_Backpack, sortPriority=150}
  n73 = MaterialFunctionCall(Float2=NamedRerouteUsage()) {materialFunction=&BreakOutFloat2Components}
  NamedRerouteDeclaration(None=LinearInterpolate(
    A=LinearInterpolate(A=@n71, B=@n72, Alpha=@n73.R)
    B=ScalarParameter() {defaultValue=1, group=Effect_Visor_UVS, parameterName=EVisor_Side_Scale_Pickaxe, sortPriority=150}
    Alpha=@n73.G
   )) {name=Side_Scale, nodeColor=<rgba 0 0.7188 1 1>}
  n74 = ScalarParameter() {group=Effect_Visor_UVS, parameterName=EVisor_Side_Rotation_Vehicle, sortPriority=150}
  n75 = ScalarParameter() {group=Effect_Visor_UVS, parameterName=EVisor_Side_Rotation_Vehicle_Side, sortPriority=150}
  n76 = MaterialFunctionCall(Float4=NamedRerouteUsage()) {materialFunction=&BreakOutFloat4Components}
  n77 = LinearInterpolate(
   A=LinearInterpolate(
    A=ScalarParameter() {group=Effect_Visor_UVS, parameterName=EVisor_Side_Rotation_Weapon, sortPriority=150}
    B=LinearInterpolate(A=@n74, B=@n75, Alpha=@n76.G)
    Alpha=@n76.R
   )
   B=LinearInterpolate(
    A=ScalarParameter() {group=Effect_Visor_UVS, parameterName=EVisor_Side_Rotation_Mech, sortPriority=150}
    B=ScalarParameter() {group=Effect_Visor_UVS, parameterName=EVisor_Side_Rotation_Character, sortPriority=150}
    Alpha=@n76.A
   )
   Alpha=@n76.B
  )
  n78 = ScalarParameter() {group=Effect_Visor_UVS, parameterName=EVisor_Side_Rotation_Backpack, sortPriority=150}
  n79 = MaterialFunctionCall(Float2=NamedRerouteUsage()) {materialFunction=&BreakOutFloat2Components}
  NamedRerouteDeclaration(None=LinearInterpolate(
    A=LinearInterpolate(A=@n77, B=@n78, Alpha=@n79.R)
    B=ScalarParameter() {group=Effect_Visor_UVS, parameterName=EVisor_Side_Rotation_Pickaxe, sortPriority=150}
    Alpha=@n79.G
   )) {name=Side_Rotation, nodeColor=<rgba 0 0.7188 1 1>}
  n80 = ScalarParameter() {group=Effect_Visor_Depth, parameterName=EVisor_Depth_Offset_Vehicle, sortPriority=150}
  n81 = ScalarParameter() {group=Effect_Visor_Depth, parameterName=EVisor_Depth_Offset_Vehicle_Side, sortPriority=150}
  n82 = MaterialFunctionCall(Float4=NamedRerouteUsage()) {materialFunction=&BreakOutFloat4Components}
  n83 = LinearInterpolate(
   A=LinearInterpolate(
    A=ScalarParameter() {group=Effect_Visor_Depth, parameterName=EVisor_Depth_Offset_Weapon, sortPriority=150}
    B=LinearInterpolate(A=@n80, B=@n81, Alpha=@n82.G)
    Alpha=@n82.R
   )
   B=LinearInterpolate(
    A=ScalarParameter() {group=Effect_Visor_Depth, parameterName=EVisor_Depth_Offset_Mech, sortPriority=150}
    B=ScalarParameter() {group=Effect_Visor_Depth, parameterName=EVisor_Depth_Offset_Character, sortPriority=150}
    Alpha=@n82.A
   )
   Alpha=@n82.B
  )
  n84 = ScalarParameter() {group=Effect_Visor_Depth, parameterName=EVisor_Depth_Offset_Backpack, sortPriority=150}
  n85 = MaterialFunctionCall(Float2=NamedRerouteUsage()) {materialFunction=&BreakOutFloat2Components}
  NamedRerouteDeclaration(None=LinearInterpolate(
    A=LinearInterpolate(A=@n83, B=@n84, Alpha=@n85.R)
    B=ScalarParameter() {group=Effect_Visor_Depth, parameterName=EVisor_Depth_Offset_Pickaxe, sortPriority=150}
    Alpha=@n85.G
   )) {name=Side_Depth, nodeColor=<rgba 0 0.7188 1 1>}
  n86 = NamedRerouteUsage()
  n87 = Add(
   A=NamedRerouteUsage()
   B=Multiply(A=Add(A=@n86) {constB=-0.5}, B=ScalarParameter() {group=Effect_Visor, parameterName=EVisor_Cuvatur_Warp_amount, sortPriority=0})
  )
  n88 = CameraVectorWS()
  n89 = Multiply(
   A=Multiply(
    A=Transform(None=@n88) {transformSourceType=TRANSFORMSOURCE_World, transformType=TRANSFORM_Local}
    B=MaterialFunctionCall() {materialFunction=&ObjectScale}.Scale Z
   )
   B=NamedRerouteUsage()
  )
  n90 = Multiply(A=@n89) {constB=-0.001}
  n91 = MaterialFunctionCall(Color=NamedRerouteUsage()) {materialFunction=&SplitComponents}
  NamedRerouteDeclaration(None=MaterialFunctionCall!(
    For_Safe_Zones=StaticBool() {value=true}
    Normal_Map
    Caculated_Bounds=Add(A=@n87, B=@n90)
    Pre_Skinned_Normals=NamedRerouteUsage()
    Is Vehicle=@n91.R
    Is VehicleSide=@n91.R
    Is Mech=@n91.R
    Vehicle_Rotation
    UV_Rotation Bypass=NamedRerouteUsage()
    Mech_Rotation
    Uvs_Offset=NamedRerouteUsage()
    Offset-after rotation
    Uvs_TileRate=NamedRerouteUsage()
   ) {materialFunction=&/MF_Projection_Side_Normal}.UVs) {name=Uvs_Layer02, nodeColor=<rgba 0.3359 0 1 1>}
  n92 = VectorParameter() {group=Effect_Visor_UVS, parameterName=EVisor_Side_Offset_Vehicle_01}
  n93 = VectorParameter() {group=Effect_Visor_UVS, parameterName=EVisor_Side_Offset_VehicleSide_01}
  n94 = MaterialFunctionCall(Float4=NamedRerouteUsage()) {materialFunction=&BreakOutFloat4Components}
  n95 = LinearInterpolate(
   A=LinearInterpolate(
    A=VectorParameter() {group=Effect_Visor_UVS, parameterName=EVisor_Side_Offset_Weapon_01}.RGB
    B=LinearInterpolate(A=@n92.RGB, B=@n93.RGB, Alpha=@n94.G)
    Alpha=@n94.R
   )
   B=LinearInterpolate(
    A=VectorParameter() {group=Effect_Visor_UVS, parameterName=EVisor_Side_Offset_Mech_01}.RGB
    B=VectorParameter() {group=Effect_Visor_UVS, parameterName=EVisor_Side_Offset_Character_01}.RGB
    Alpha=@n94.A
   )
   Alpha=@n94.B
  )
  n96 = VectorParameter() {group=Effect_Visor_UVS, parameterName=EVisor_Side_Offset_Backpack_01}
  n97 = MaterialFunctionCall(Float2=NamedRerouteUsage()) {materialFunction=&BreakOutFloat2Components}
  NamedRerouteDeclaration(None=LinearInterpolate(
    A=LinearInterpolate(A=@n95, B=@n96.RGB, Alpha=@n97.R)
    B=VectorParameter() {group=Effect_Visor_UVS, parameterName=EVisor_Side_Offset_Pickaxe_01}.RGB
    Alpha=@n97.G
   )) {name=Side_offset, nodeColor=<rgba 0 0.625 1 1>}
  n98 = TextureObjectParameter() {
   group=Effect_Visor
   parameterName=EVisor__Texture_MSK
   samplerType=SAMPLERTYPE_Masks
   sortPriority=1
   texture=/Game/Weapons/WeaponSkins/Wraps/Textures/OriginPrison/T_Wrap_OriginPrisoner_MSK2
  }
  NamedRerouteDeclaration(None=@n98) {name=Texture_MSK, nodeColor=<rgba 0 1 0.7188 1>}
  n99 = FunctionInput(Preview=@n62.G) {bUsePreviewValueAsDefault=true, inputName=Mask Crevise, inputType=FunctionInput_Scalar, sortPriority=24, previewValue=<1 0 0 1>}
  NamedRerouteDeclaration(None=@n99) {name=Crevise_map, nodeColor=<rgba 1 0.6797 0 1>}
  NamedRerouteDeclaration(None=FunctionInput(Preview=Constant4Vector().RGBA) {bUsePreviewValueAsDefault=true, inputName=IsState, inputType=FunctionInput_Vector4, sortPriority=4}) {name=IsState, nodeColor=<rgba 0.1497 1 0.1633 1>}
  NamedRerouteDeclaration(None=FunctionInput(Preview=Constant2Vector()) {bUsePreviewValueAsDefault=true, inputName=IsState_B, inputType=FunctionInput_Vector2, sortPriority=4}) {name=IsState_B, nodeColor=<rgba 0.1497 1 0.1633 1>}
  n100 = PreSkinnedLocalBounds()
  n101 = Subtract(A=VertexInterpolator(VS=PreSkinnedPosition()).PS, B=@n100.Min)
  n102 = MaterialFunctionCall(Color=Subtract(A=@n100.Max, B=@n100.Max)) {materialFunction=&SplitComponents}
  n103 = Max(A=Max(A=@n102.R, B=@n102.R), B=@n102.B)
  NamedRerouteDeclaration(None=FunctionInput(Preview=Divide(A=@n101, B=@n103)) {bUsePreviewValueAsDefault=true, inputName=Calcated_Bounds, sortPriority=11}) {name=Caclcated_Bounds, nodeColor=<rgba 0 0.9297 1 1>}
  NamedRerouteDeclaration(None=ComponentMask(None=@n2.RGB) {g=true, r=true}) {name=SubImages, nodeColor=<rgba 0 1 0.9297 1>}
  n104 = PreSkinnedNormal()
  NamedRerouteDeclaration(None=FunctionInput(Preview=VertexInterpolator(VS=@n104).PS) {bUsePreviewValueAsDefault=true, inputName=Pre-Skinned Local Normal, sortPriority=10}) {name=PreSkin_Normal, nodeColor=<rgba 0.8281 1 0 1>}
  n105 = Multiply(A=@n89) {constB=-0.002}
  n106 = MaterialFunctionCall(Color=NamedRerouteUsage()) {materialFunction=&SplitComponents}
  NamedRerouteDeclaration(None=MaterialFunctionCall(
    For_Safe_Zones=StaticBool() {value=true}
    Caculated_Bounds=Add(A=@n87, B=@n105)
    Pre_Skinned_Normals=NamedRerouteUsage()
    Is Vehicle=@n106.R
    Is VehicleSide=@n106.R
    Is Mech=@n106.R
    UV_Rotation Bypass=NamedRerouteUsage()
    Uvs_Offset=NamedRerouteUsage()
    Uvs_TileRate=NamedRerouteUsage()
   ) {materialFunction=&MF_Projection_Side_Normal}.UVs) {name=Uvs_Layer03, nodeColor=<rgba 0.3359 0 1 1>}
  n107 = StaticBool() {value=true}
  n108 = NamedRerouteUsage()
  n109 = MaterialFunctionCall(Color=NamedRerouteUsage()) {materialFunction=&SplitComponents}
  n110 = NamedRerouteUsage()
  n111 = NamedRerouteUsage()
  n112 = NamedRerouteUsage()
  TextureSample(UVs=NamedRerouteDeclaration(None=MaterialFunctionCall(
     For_Safe_Zones=@n107
     Caculated_Bounds=@n87
     Pre_Skinned_Normals=@n108
     Is Vehicle=@n109.R
     Is VehicleSide=@n109.R
     Is Mech=@n109.R
     UV_Rotation Bypass=@n110
     Uvs_Offset=@n111
     Uvs_TileRate=@n112
    ) {materialFunction=&MF_Projection_Side_Normal}.UVs) {name=Uvs_Layer01, nodeColor=<rgba 0.3359 0 1 1>}) {texture=/Game/Weapons/WeaponSkins/Wraps/Textures/UVCheaker}
 ]
}
