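# Material function for the FryAngles weapon-skin wrap effect, layering two timed emissive masks with optional base darkening onto material attributes for weapons, vehicles and mechs.
# Wrap bounds and pre-skinned normals drive top/side mask projections, which warp and modulate gradient-sampled masks.
# FA_TimeFrom_BP starts eased color, brightness and gradient-offset transitions, with layer two colored through a curve atlas.
# A static switch enables the result, blended by the customization mask.
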MaterialFunction /Game/Weapons/WeaponSkins/Wraps/MaterialFunctions/MF_Fryangles
graph {nodes [
  n1 = FunctionInput(Preview=VertexInterpolator(VS=PreSkinnedNormal()).PS) {bUsePreviewValueAsDefault=true, inputName=Pre-Skinned Local Normal, sortPriority=10}
  n2 = ComponentMask(None=@n1) {g=true}
  Abs(None=@n2)
  n3 = FunctionInput() {inputName=Material, inputType=FunctionInput_MaterialAttributes, sortPriority=-20}
  n4 = GetMaterialAttributes(None=@n3) {attributeGetTypes [#1, #2]}
  n5 = ScalarParameter() {group=FryAngles, parameterName=FA-Emissive01-DarkenBase-------, sortPriority=50}
  n6 = TextureObjectParameter() {group=FryAngles, parameterName=FA_Mask_01, samplerType=SAMPLERTYPE_Masks, texture=&T_BlackMask}
  n7 = FunctionInput(Preview=Constant()) {bUsePreviewValueAsDefault=true, inputName=isMech, inputType=FunctionInput_Scalar, sortPriority=4}
  n8 = VectorParameter() {group=FryAngles, parameterName=FA_Fill_alternate_Axis_Color_Mask_01, defaultValue=<rgba 0 0 0 1>}
  n9 = Constant()
  n10 = FunctionInput(Preview=@n9) {inputName=isVehicle, inputType=FunctionInput_Scalar, sortPriority=5}
  n11 = VectorParameter() {group=FryAngles_UVS, parameterName=TS-Rotation-Weapon-01---------------------, sortPriority=1, defaultValue=<rgba 0 0 -0.25 1>}
  n12 = VectorParameter() {group=FryAngles_UVS, parameterName=FA_Rotation_Vehicle_01, sortPriority=2, defaultValue=<rgba 0 0 0 1>}
  n13 = VectorParameter() {group=FryAngles_UVS, parameterName=FA_Rotation_Mech_01, sortPriority=3, defaultValue=<rgba 0.25 0.25 0 1>}
  n14 = VectorParameter() {group=FryAngles_UVS, parameterName=FA_Offset_Weapon_01---------------------, sortPriority=10, defaultValue=<rgba 0 0 0 1>}
  n15 = VectorParameter() {group=FryAngles_UVS, parameterName=FA_Offset_Vehicle_01, sortPriority=11, defaultValue=<rgba 0 0 0 1>}
  n16 = VectorParameter() {group=FryAngles_UVS, parameterName=FA_Offset_Mech_01, sortPriority=12, defaultValue=<rgba 0 0 0 1>}
  n17 = PreSkinnedPosition()
  n18 = MaterialFunctionCall!(
   is Vehicle=@n10
   Object Bounds Max Adjustment=FunctionInput() {bUsePreviewValueAsDefault=true, inputName=Object Bounds Max Adjustment, sortPriority=13}
   Object Bounds Min Adjustment=FunctionInput() {bUsePreviewValueAsDefault=true, inputName=Object Bounds Min Adjustment, sortPriority=12}
   Object Bounds Center Adjustment=FunctionInput() {bUsePreviewValueAsDefault=true, inputName=Object Bounds Center Adjustment, sortPriority=14}
   isMech=@n7
   Side Project Vehicle
   Local Position=FunctionInput(Preview=VertexInterpolator(VS=@n17).PS) {bUsePreviewValueAsDefault=true, inputName=Local Position, sortPriority=11}
  ) {materialFunction=&/MF_WrapBounds}
  n19 = Add(A=@n18.Result, B=LinearInterpolate(A=LinearInterpolate(A=@n14.RGB, B=@n15.RGB, Alpha=@n10), B=@n16.RGB, Alpha=@n7))
  n20 = VectorParameter() {group=FryAngles_UVS, parameterName=FA_Panning_Weapon_01---------------------, sortPriority=40, defaultValue=<rgba 0 0 0 1>}
  n21 = VectorParameter() {group=FryAngles_UVS, parameterName=FA_Panning_Vehicle_01, sortPriority=41, defaultValue=<rgba 0 0 0 1>}
  n22 = VectorParameter() {group=FryAngles_UVS, parameterName=FA_Panning_Mech_01, sortPriority=42, defaultValue=<rgba 0 0 0 1>}
  n23 = Multiply(A=Time(), B=LinearInterpolate(A=LinearInterpolate(A=@n20.RGB, B=@n21.RGB, Alpha=@n10), B=@n22.RGB, Alpha=@n7))
  n24 = VectorParameter() {group=FryAngles_UVS, parameterName="FA_Scale_Weapon_01 A=uniform", sortPriority=30, defaultValue=<rgba 1 1 1 1>}
  n25 = VectorParameter() {group=FryAngles_UVS, parameterName="FA_Scale_Vehicle_01 A=uniform", sortPriority=31, defaultValue=<rgba 1 1 1 1>}
  n26 = LinearInterpolate(A=Multiply(A=@n24.RGB, B=@n24.RGB), B=Multiply(A=@n25.RGB, B=@n25.RGB), Alpha=@n10)
  n27 = VectorParameter() {group=FryAngles_UVS, parameterName="FA_Scale_Mech_01 A=uniform", defaultValue=<rgba 1 1 1 1>}
  n28 = Multiply(A=@n27.RGB, B=@n27.RGB)
  n29 = MaterialFunctionCall!(
   IsCharacter
   IsStatsB
   Texture=@n6
   Rotation_Axis=LinearInterpolate(A=LinearInterpolate(A=@n11.RGB, B=@n12.RGB, Alpha=@n10), B=@n13.RGB, Alpha=@n7)
   Caculated_Bounds=Multiply(A=Add(A=@n19, B=@n23), B=LinearInterpolate(A=@n26, B=@n28, Alpha=@n7))
   Pre_Skinned_Normals=@n1
   Blend_Contrast=LinearInterpolate(
    A=ScalarParameter() {defaultValue=1, group=FryAngles, parameterName=FA_Blend_Contrast}
    B=ScalarParameter() {defaultValue=2, group=FryAngles, parameterName=FA_Blend_Contrast_Mech}
    Alpha=@n7
   )
   Fill_alternate_Axis_Opacity=ScalarParameter() {group=FryAngles, parameterName=FA_Fill_alternate_Axis_Opacity 01}
   "Fill_alternate_Axis_Color_A=Roughness"=AppendVector(A=@n8.RGB, B=@n8.RGB)
   Is_Vehicle=@n10
   Is_Mech=@n7
   Enable_Fill_Z_Weapon
   Fill_Z_Weapon_opacity
   Fill_Z_Weapon_Color
  ) {materialFunction=&/MF_TopsSideProjection_Base}
  n30 = VectorParameter() {group=FryAngles, parameterName=FA_Emissive01_OffsetWeapon, sortPriority=55, defaultValue=<rgba 0 -1 0 0>}
  n31 = VectorParameter() {group=FryAngles, parameterName=FA_Emissive01_OffsetVehicle, sortPriority=55, defaultValue=<rgba -1 0 0 0>}
  n32 = VectorParameter() {group=FryAngles, parameterName=FA_Emissive01_OffsetMech, sortPriority=55, defaultValue=<rgba 0 0 -1 0>}
  n33 = VectorParameter() {group=FryAngles, parameterName=FA_Emissive01_Spread_XY, sortPriority=54, defaultValue=<rgba 1 1 1 0>}
  n34 = VectorParameter() {group=FryAngles, parameterName=FA_Emissive01_Spread_Mech_XY, sortPriority=54, defaultValue=<rgba 1 1 1 0>}
  n35 = VectorParameter() {group=FryAngles, parameterName=FA_Emissive01_Animated_WeaponXY, sortPriority=58}
  n36 = VectorParameter() {group=FryAngles, parameterName=FA_Emissive01_Animated_VehicleXY, sortPriority=58}
  n37 = FunctionInput(Preview=Constant3Vector() {constant=<rgba 1 0 0 0>}) {bUsePreviewValueAsDefault=true, inputName=ForwardAxis}
  n38 = FunctionInput(Preview=@n9) {bUsePreviewValueAsDefault=true, inputName=Side Project Vehicle, inputType=FunctionInput_Scalar}
  n39 = MaterialFunctionCall!(
   ForwardAxis=@n37
   Bounds=@n18.Result
   Pre-Skinned Local Position
   isVehicle=@n10
   isMech=@n7
   Side Project Vehicle=@n38
   WeaponOffset=@n30.RGB
   WeaponFalloff=ScalarParameter() {defaultValue=0.5, group=FryAngles, parameterName="FA_Emissive01_FalloffWeapon (negative inverts)", sortPriority=56}
   VehicleOffset=@n31.RGB
   VehicleFalloff=ScalarParameter() {defaultValue=0.5, group=FryAngles, parameterName="FA_Emissive01_FalloffVehicle ", sortPriority=56}
   MechOffset=@n32.RGB
   MechFalloff=ScalarParameter() {defaultValue=0.5, group=FryAngles, parameterName="FA_Emissive01_FalloffMech ", sortPriority=56}
   Invert_Gradent
   UV_Spread_XY=LinearInterpolate(A=@n33.RGB, B=@n34.RGB, Alpha=@n7)
   UV_Animated_XY=LinearInterpolate(
    A=LinearInterpolate(A=@n35.RGB, B=@n36.RGB, Alpha=@n10)
    B=VectorParameter() {group=FryAngles, parameterName=FA_Emissive01_Animated_MechXY, sortPriority=58}.RGB
    Alpha=@n7
   )
   Warp_UVs=Multiply(
    A=ChannelMaskParameter(None=@n29.Combined) {group=FryAngles, parameterName=FA_WarpEmissive01MaskChannel_01, sortPriority=59}.R
    B=ScalarParameter() {group=FryAngles, parameterName=FA_WarpEmissive01_Warp_Amount, sortPriority=59}
   )
   Rotate_Mech=ScalarParameter() {defaultValue=1, group=FryAngles, parameterName=FA_Emissive01_Texture_Rotate_Mech, sortPriority=60}
   Rotate_Vehicle_Side
   Rotate_Vehicle=ScalarParameter() {defaultValue=1, group=FryAngles, parameterName=FA_Emissive01_Texture_Rotate_Vehicle, sortPriority=60}
  ) {materialFunction=&/MF_Wrap_Gradent_Creation}
  n40 = Abs(None=ComponentMask(None=@n1) {b=true})
  n41 = ScalarParameter() {defaultValue=1, group=FryAngles, parameterName=FA_Blend_Contrast, sortPriority=48}
  n42 = Saturate(None=@n2)
  n43 = Saturate(None=Add(
    A=MaterialFunctionCall(In=@n40, Contrast=@n41) {materialFunction=&CheapContrast}.Result
    B=MaterialFunctionCall(In=@n42, Contrast=@n41) {materialFunction=&CheapContrast}.Result
   ))
  n44 = MaterialFunctionCall(In=Saturate(None=Multiply(A=@n2) {constB=-1})) {materialFunction=&CheapContrast}
  n45 = ComponentMask(None=@n1) {r=true}
  n46 = Add(
   A=Saturate(None=Subtract(A=@n43, B=@n44.Result))
   B=MaterialFunctionCall(In=Abs(None=@n45), Contrast=@n41) {materialFunction=&CheapContrast}.Result
  )
  n47 = Saturate(None=LinearInterpolate(A=Saturate(None=@n46), B=Constant() {r=1}, Alpha=@n10))
  n48 = Multiply(
   A=ChannelMaskParameter(None=TextureSample(UVs=@n39.Uvs, Tex=@n6).RGBA) {desc="Uses FA_Mask_01 ", group=FryAngles, parameterName=FA_Emissive01Gradent_TextureChannel, sortPriority=50}.R
   B=@n47
  )
  n49 = Multiply(A=ChannelMaskParameter(None=@n29.Combined) {group=FryAngles, parameterName=FA_Emissive01_X+X_MaskChannel_01, sortPriority=50}.R, B=@n48)
  n50 = Saturate(None=Multiply(A=@n49, B=Add(A=@n49, B=@n48)))
  n51 = OneMinus(None=Saturate(None=Multiply(A=@n5, B=@n50)))
  n52 = Multiply(A=@n4.EmissiveColor, B=@n51)
  n53 = VectorParameter() {group=FryAngles, parameterName=FA_Emissive02_OffsetWeapon_Start, sortPriority=105, defaultValue=<rgba 0 -1 0 0>}
  n54 = VectorParameter() {group=FryAngles, parameterName=FA_Emissive02_OffsetWeapon_End, sortPriority=105, defaultValue=<rgba 0 -1 0 0>}
  n55 = Subtract(A=Time(), B=ScalarParameter() {group=FryAngles, parameterName=FA_TimeFrom_BP, sortPriority=199})
  n56 = ScalarParameter() {defaultValue=0.8, group=FryAngles, parameterName=FA_Time_Emissive02_Duration, sortPriority=250}
  n57 = Power(
   Base=Saturate(None=Divide(A=@n55, B=@n56))
   Exp=ScalarParameter() {defaultValue=0.8, group=FryAngles, parameterName=FA_Time_Emissive02_Curve, sortPriority=251}
  )
  n58 = VectorParameter() {group=FryAngles, parameterName=FA_Emissive02_OffsetVehicle_Start, sortPriority=105, defaultValue=<rgba -1 0 0 0>}
  n59 = VectorParameter() {group=FryAngles, parameterName=FA_Emissive02_OffsetVehicle_End, sortPriority=105, defaultValue=<rgba -1 0 0 0>}
  n60 = VectorParameter() {group=FryAngles, parameterName=FA_Emissive02_OffsetMech_Start, sortPriority=105, defaultValue=<rgba 0 0 -1 0>}
  n61 = VectorParameter() {group=FryAngles, parameterName=FA_Emissive02_OffsetMech_End, sortPriority=105, defaultValue=<rgba 0 0 -1 0>}
  n62 = VectorParameter() {group=FryAngles, parameterName=FA_Emissive02_Spread_XY, sortPriority=104, defaultValue=<rgba 1 1 1 0>}
  n63 = VectorParameter() {group=FryAngles, parameterName=FA_Emissive02_Spread_Mech_XY, sortPriority=104, defaultValue=<rgba 1 1 1 0>}
  n64 = LinearInterpolate(
   A=VectorParameter() {group=FryAngles, parameterName=FA_Emissive02_Animated_WeaponXY, sortPriority=108}.RGB
   B=VectorParameter() {group=FryAngles, parameterName=FA_Emissive02_Animated_VehicleXY, sortPriority=108}.RGB
   Alpha=@n10
  )
  n65 = VectorParameter() {group=FryAngles, parameterName=FA_Emissive02_Animated_MechXY, sortPriority=108}
  n66 = TextureObjectParameter() {
   group=FryAngles
   parameterName=FA_Mask_02
   samplerType=SAMPLERTYPE_Masks
   sortPriority=20
   texture=&T_BlackMask
  }
  n67 = VectorParameter() {group=FryAngles, parameterName=FA_Fill_alternate_Axis_Color_Mask 02, sortPriority=20, defaultValue=<rgba 0 0 0 1>}
  n68 = VectorParameter() {group=FryAngles_UVS, parameterName=TS-Rotation-Weapon-02---------------------, sortPriority=150, defaultValue=<rgba 0 0 -0.25 1>}
  n69 = VectorParameter() {group=FryAngles_UVS, parameterName=FA_Rotation_Vehicle_02, sortPriority=151, defaultValue=<rgba 0 0 0 1>}
  n70 = VectorParameter() {group=FryAngles_UVS, parameterName=FA_Rotation_Mech_02, sortPriority=153, defaultValue=<rgba 0.25 0.25 0 1>}
  n71 = VectorParameter() {group=FryAngles_UVS, parameterName=FA_Offset_Weapon_02---------------------, sortPriority=160, defaultValue=<rgba 0 0 0 1>}
  n72 = VectorParameter() {group=FryAngles_UVS, parameterName=FA_Offset_Vehicle_02, sortPriority=161, defaultValue=<rgba 0 0 0 1>}
  n73 = VectorParameter() {group=FryAngles_UVS, parameterName=FA_Offset_Mech_02, sortPriority=162, defaultValue=<rgba 0 0 0 1>}
  n74 = Add(A=@n18.Result, B=LinearInterpolate(A=LinearInterpolate(A=@n71.RGB, B=@n72.RGB, Alpha=@n10), B=@n73.RGB, Alpha=@n7))
  n75 = VectorParameter() {group=FryAngles_UVS, parameterName=FA_Panning_Weapon_02---------------------, sortPriority=180, defaultValue=<rgba 0 0 0 1>}
  n76 = VectorParameter() {group=FryAngles_UVS, parameterName=FA_Panning_Vehicle_02, sortPriority=181, defaultValue=<rgba 0 0 0 1>}
  n77 = VectorParameter() {group=FryAngles_UVS, parameterName=FA_Panning_Mech_02, sortPriority=182, defaultValue=<rgba 0 0 0 1>}
  n78 = Multiply(A=Time(), B=LinearInterpolate(A=LinearInterpolate(A=@n75.RGB, B=@n76.RGB, Alpha=@n10), B=@n77.RGB, Alpha=@n7))
  n79 = VectorParameter() {group=FryAngles_UVS, parameterName="FA_Scale_Weapon_02 A=uniform", sortPriority=170, defaultValue=<rgba 1 1 1 1>}
  n80 = VectorParameter() {group=FryAngles_UVS, parameterName="FA_Scale_Vehicle_02 A=uniform", sortPriority=171, defaultValue=<rgba 1 1 1 1>}
  n81 = LinearInterpolate(A=Multiply(A=@n79.RGB, B=@n79.RGB), B=Multiply(A=@n80.RGB, B=@n80.RGB), Alpha=@n10)
  n82 = VectorParameter() {group=FryAngles_UVS, parameterName="FA_Scale_Mech_02 A=uniform", sortPriority=172, defaultValue=<rgba 1 1 1 1>}
  n83 = Multiply(A=@n82.RGB, B=@n82.RGB)
  n84 = MaterialFunctionCall(
   Texture=@n66
   Rotation_Axis=LinearInterpolate(A=LinearInterpolate(A=@n68.RGB, B=@n69.RGB, Alpha=@n10), B=@n70.RGB, Alpha=@n7)
   Caculated_Bounds=Multiply(A=Add(A=@n74, B=@n78), B=LinearInterpolate(A=@n81, B=@n83, Alpha=@n7))
   Pre_Skinned_Normals=@n1
   Blend_Contrast=LinearInterpolate(
    A=ScalarParameter() {defaultValue=1, group=FryAngles, parameterName=FA_Blend_Contrast, sortPriority=20}
    B=ScalarParameter() {defaultValue=2, group=FryAngles, parameterName=FA_Blend_Contrast_Mech, sortPriority=20}
    Alpha=@n7
   )
   Fill_alternate_Axis_Opacity=ScalarParameter() {group=FryAngles, parameterName=FA_Fill_alternate_Axis_Opacity 02, sortPriority=20}
   "Fill_alternate_Axis_Color_A=Roughness"=AppendVector(A=@n67.RGB, B=@n67.RGB)
   Is_Vehicle=@n10
   Is_Mech=@n7
  ) {materialFunction=&MF_TopsSideProjection_Base}
  n85 = Multiply(
   A=ChannelMaskParameter(None=@n84.Combined) {group=FryAngles, parameterName=FA_WarpEmissive02MaskChannel_02, sortPriority=109}.R
   B=ScalarParameter() {group=FryAngles, parameterName=FA_WarpEmissive02_Warp_Amount, sortPriority=109}
  )
  n86 = TextureSample(
   UVs=MaterialFunctionCall(
    ForwardAxis=@n37
    Bounds=@n18.Result
    isVehicle=@n10
    isMech=@n7
    Side Project Vehicle=@n38
    WeaponOffset=LinearInterpolate(A=@n53.RGB, B=@n54.RGB, Alpha=@n57)
    WeaponFalloff=ScalarParameter() {defaultValue=0.5, group=FryAngles, parameterName="FA_Emissive02_FalloffWeapon (negative inverts)", sortPriority=106}
    VehicleOffset=LinearInterpolate(A=@n58.RGB, B=@n59.RGB, Alpha=@n57)
    VehicleFalloff=ScalarParameter() {defaultValue=0.5, group=FryAngles, parameterName="FA_Emissive02_FalloffVehicle ", sortPriority=106}
    MechOffset=LinearInterpolate(A=@n60.RGB, B=@n61.RGB, Alpha=@n57)
    MechFalloff=ScalarParameter() {defaultValue=0.5, group=FryAngles, parameterName=FA_Emissive02_FalloffMech, sortPriority=106}
    UV_Spread_XY=LinearInterpolate(A=@n62.RGB, B=@n63.RGB, Alpha=@n7)
    UV_Animated_XY=LinearInterpolate(A=@n64, B=@n65.RGB, Alpha=@n7)
    Warp_UVs=@n85
    Rotate_Mech=ScalarParameter() {defaultValue=1, group=FryAngles, parameterName=FA_Emissive02_Texture_Rotate_Mech, sortPriority=110}
    Rotate_Vehicle=ScalarParameter() {defaultValue=1, group=FryAngles, parameterName=FA_Emissive02_Texture_Rotate_Vehicle, sortPriority=110}
   ) {materialFunction=&MF_Wrap_Gradent_Creation}.Uvs
   Tex=@n66
  ) {samplerType=SAMPLERTYPE_Masks}
  n87 = MaterialFunctionCall(
   Transition Phase=Multiply(
    A=@n47
    B=ChannelMaskParameter(None=@n86.RGBA) {desc=Uses_FA_Mask_02, group=FryAngles, parameterName=FA_Emissive02_Gradent_TextureChannel, sortPriority=100}.R
   )
   Height Texture=ChannelMaskParameter(None=@n84.Combined) {group=FryAngles, parameterName=FA_Emissive02_HeightLerp_MaskChannel_02, sortPriority=100}.R
   Contrast=ScalarParameter() {group=FryAngles, parameterName=FA_Emissive02_HeightLerp_Contrast, sortPriority=100}
  ) {materialFunction=&HeightLerp}
  n88 = ScalarParameter() {group=FryAngles, parameterName=FA-Emissive02-DarkenBase-------, sortPriority=99}
  n89 = OneMinus(None=Saturate(None=Multiply(A=@n87.Alpha, B=@n88)))
  n90 = VectorParameter() {group=FryAngles, parameterName=FA_Time_Emissive01_Color_Start, sortPriority=202, defaultValue=<rgba 50 0 0 1>}
  n91 = VectorParameter() {group=FryAngles, parameterName=FA_Time_Emissive01_Color_End, sortPriority=203, defaultValue=<rgba 1.5 0 0 1>}
  n92 = Subtract(A=Time(), B=ScalarParameter() {group=FryAngles, parameterName=FA_TimeFrom_BP, sortPriority=199})
  n93 = ScalarParameter() {defaultValue=0.8, group=FryAngles, parameterName=FA_Time_Emissive01_Duration, sortPriority=200}
  n94 = Power(
   Base=Saturate(None=Divide(A=@n92, B=@n93))
   Exp=ScalarParameter() {defaultValue=0.8, group=FryAngles, parameterName=FA_Time_Emissive01_Curve, sortPriority=201}
  )
  n95 = Multiply(A=Multiply(A=LinearInterpolate(A=@n90.RGB, B=@n91.RGB, Alpha=@n94), B=@n50), B=@n89)
  n96 = ScalarParameter() {group=FryAngles, parameterName=FA_Emissive02_Curve_Offset_Weapon, sortPriority=100}
  n97 = ScalarParameter() {group=FryAngles, parameterName=FA_Emissive02_Curve_Offset_Vehicle, sortPriority=100}
  n98 = ScalarParameter() {defaultValue=1, desc="negative value inverts ", group=FryAngles, parameterName=FA_Emissive02_Curve_Falloff_Weapon, sortPriority=100}
  n99 = ScalarParameter() {defaultValue=1, desc="negative value inverts ", group=FryAngles, parameterName=FA_Emissive02_Curve_Falloff_Vehicle, sortPriority=100}
  n100 = ScalarParameter() {defaultValue=1, desc="negative value inverts ", group=FryAngles, parameterName=FA_Emissive02_Curve_Falloff_Mech, sortPriority=100}
  n101 = MaterialFunctionCall!(
   Bounds=@n18.Result
   ForwardAxis=@n37
   Offset=LinearInterpolate(
    A=LinearInterpolate(A=@n96, B=@n97, Alpha=@n10)
    B=ScalarParameter() {group=FryAngles, parameterName=FA_Emissive02_Curve_Offset_Mech, sortPriority=100}
    Alpha=@n7
   )
   FallOff=LinearInterpolate(A=LinearInterpolate(A=@n98, B=@n99, Alpha=@n10), B=@n100, Alpha=@n7)
  ) {materialFunction=&/MF_Wrap_Gradent_Simple}
  n102 = Frac(None=Add(A=@n85, B=Frac(None=@n101.Raw_Data)))
  n103 = CurveAtlasRowParameter(CurveTime=@n102) {
   atlas=/Game/Weapons/WeaponSkins/Wraps/Textures/CruveAtlas_Wraps_01
   curve=/Game/Weapons/WeaponSkins/Wraps/Textures/C_GradSimple
   group=FryAngles
   parameterName=FA_Emissive02_Curve
   sortPriority=100
  }
  n104 = MaterialFunctionCall!(
   Index=ScalarParameter() {group=FryAngles, parameterName="FA_Emissive02_Curve  Index", sortPriority=100}
   CCinput=@n103
   CCinput_Alpha=@n103
   UV=@n102
  ) {materialFunction=&/MF_Global_CC_Tex}
  n105 = LinearInterpolate(
   A=ScalarParameter() {defaultValue=50, group=FryAngles, parameterName=FA_Time_Emissive02_CurveBrightness_Start, sortPriority=255}
   B=ScalarParameter() {defaultValue=1.5, group=FryAngles, parameterName=FA_Time_Emissive02_CurveBrightness_End, sortPriority=256}
   Alpha=@n57
  )
  n106 = Multiply(A=@n87.Alpha, B=Multiply(A=@n104.3V, B=Multiply(A=@n104.Alpha, B=@n105)))
  n107 = SetMaterialAttributes!(
   MaterialAttributes=@n4.MaterialAttributes
   EmissiveColor=Add(A=Multiply(A=@n52, B=@n89), B=Add(A=@n95, B=@n106))
   BaseColor=Multiply(A=Multiply(A=@n4.BaseColor, B=@n51), B=@n89)
  ) {attributeSetTypes [#2, #1]}
  n108 = FunctionInput() {bUsePreviewValueAsDefault=true, inputName="Customization Mask (_CM B)", inputType=FunctionInput_Scalar, sortPriority=24}
  FunctionOutput(None=StaticSwitchParameter(True=BlendMaterialAttributes(A=@n4.MaterialAttributes, B=@n107, Alpha=@n108), False=@n3) {group=Customization_Wrap, parameterName=FryAngles})
  n109 = CurveAtlasRowParameter(CurveTime=@n102) {
   atlas=&CruveAtlas_Wraps_01
   curve=&C_GradSimple
   group=FryAngles
   parameterName=FA_Emissive02_Curve_End
   sortPriority=100
  }
  n110 = MaterialFunctionCall(
   Index=ScalarParameter() {group=FryAngles, parameterName="FA_Emissive02_Curve_End  index", sortPriority=100}
   CCinput=@n109
   CCinput_Alpha=@n109
   UV=@n102
  ) {materialFunction=&MF_Global_CC_Tex}
  LinearInterpolate(A=@n104.3V, B=@n110.3V, Alpha=@n57)
  LinearInterpolate(A=@n104.Alpha, B=@n110.Alpha, Alpha=@n57)
  TextureObjectParameter() {
   group=FryAngles
   parameterName=FA_Emissive01Gradent_Texture
   samplerType=SAMPLERTYPE_Masks
   sortPriority=50
   texture=/Game/Global/Textures/Default/Masks_and_Gradients/T_VertGradient_MASK
  }
  TextureObjectParameter() {
   group=FryAngles
   parameterName=FA_Emissive02_Gradent_Texture
   samplerType=SAMPLERTYPE_Masks
   sortPriority=100
   texture=&T_VertGradient_MASK
  }
 ]}
