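# Weapon wrap material function that applies an animated flow-map pattern to weapons, vehicles and mechs.
# Bounds-based skew or side projections, chosen by static switches and tuned via isState, produce UVs for two half-cycle-offset flow samples crossfaded to hide resets.
# An axis mask blends these with a static tri-area texture for base color and roughness, and gates the added normal.
# Customization masks composite the result onto the input material, behind a bypass switch.
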MaterialFunction /Game/Weapons/WeaponSkins/Wraps/MaterialFunctions/MF_Function_FlowMap
graph {nodes [
  Constant() {r=1}
  FunctionInput(Preview=TextureSample() {samplerType=SAMPLERTYPE_Masks, texture=/Game/Weapons/FORT_Rifles/Textures/T_Scar_CM}.G) {bUsePreviewValueAsDefault=true, inputName=Curavture, inputType=FunctionInput_Scalar, sortPriority=24}
  n1 = VectorParameter() {group=Function_FlowMap, parameterName=FFlow_UV_Offset_Vehicle}
  n2 = VectorParameter() {group=Function_FlowMap, parameterName=FFlow_UV_Offset_VehicleSide}
  n3 = FunctionInput(Preview=Constant()) {bUsePreviewValueAsDefault=true, inputName=isState, inputType=FunctionInput_Vector4, sortPriority=4}
  n4 = MaterialFunctionCall(Color=@n3) {materialFunction=&SplitComponents}
  n5 = LinearInterpolate(
   A=LinearInterpolate(
    A=VectorParameter() {group=Function_FlowMap, parameterName=FFlow_UV_Offset_Weapon}.RGB
    B=LinearInterpolate(A=@n1.RGB, B=@n2.RGB, Alpha=@n4.G)
    Alpha=@n4.R
   )
   B=VectorParameter() {group=Function_FlowMap, parameterName=FFlow_UV_Offset_Mech}.RGB
   Alpha=@n4.B
  )
  n6 = VectorParameter() {desc=Skew Projection TileRate, group=Function_FlowMap, parameterName=FFlow_UV_TileRate_Weapon, defaultValue=<rgba 1 2 0 1>}
  n7 = VectorParameter() {desc=Skew Projection TileRate, group=Function_FlowMap, parameterName=FFlow_UV_TileRate_Vehicle, defaultValue=<rgba 1 2 0 1>}
  n8 = Multiply(A=@n7.RGB, B=@n7.RGB)
  n9 = VectorParameter() {desc=Skew Projection TileRate, group=Function_FlowMap, parameterName=FFlow_UV_TileRate_VehicleSide, defaultValue=<rgba 1 2 0 1>}
  n10 = Multiply(A=@n9.RGB, B=@n9.RGB)
  n11 = MaterialFunctionCall(Color=@n3) {materialFunction=&SplitComponents}
  n12 = VectorParameter() {desc=Skew Projection TileRate, group=Function_FlowMap, parameterName=FFlow_UV_TileRate_Mech, defaultValue=<rgba 1 2 0 1>}
  n13 = LinearInterpolate(
   A=LinearInterpolate(A=Multiply(A=@n6.RGB, B=@n6.RGB), B=LinearInterpolate(A=@n8, B=@n10, Alpha=@n11.G), Alpha=@n11.R)
   B=Multiply(A=@n12.RGB, B=@n12.RGB)
   Alpha=@n11.B
  )
  n14 = ScalarParameter() {defaultValue=1, group=Function_FlowMap, parameterName=FFlow_TriPlanner_Mask_Contrast, sortPriority=1}
  n15 = ScalarParameter() {defaultValue=0.5, group=Function_FlowMap, parameterName=FFlow_UV_Skew_Amount_Vehicle}
  n16 = ScalarParameter() {defaultValue=0.5, group=Function_FlowMap, parameterName=FFlow_UV_Skew_Amount_VehicleSide}
  n17 = MaterialFunctionCall(Color=@n3) {materialFunction=&SplitComponents}
  n18 = LinearInterpolate(
   A=LinearInterpolate(
    A=ScalarParameter() {defaultValue=0.5, group=Function_FlowMap, parameterName=FFlow_UV_Skew_Amount_Weapon}
    B=LinearInterpolate(A=@n15, B=@n16, Alpha=@n17.G)
    Alpha=@n17.R
   )
   B=ScalarParameter() {defaultValue=0.5, group=Function_FlowMap, parameterName="FFlow_UV_Skew_Amount_Mech "}
   Alpha=@n17.B
  )
  n19 = VertexInterpolator(VS=PreSkinnedPosition())
  n20 = PreSkinnedLocalBounds()
  n21 = FunctionInput(Preview=Divide(A=Subtract(A=@n19.PS, B=@n20.Min), B=Subtract(A=@n20.Max, B=@n20.Max))) {bUsePreviewValueAsDefault=true, inputName=Calcated_Bounds, sortPriority=11}
  n22 = FunctionInput(Preview=VertexInterpolator(VS=PreSkinnedNormal()).PS) {bUsePreviewValueAsDefault=true, inputName=Pre-Skinned Local Normal, sortPriority=10}
  n23 = MaterialFunctionCall(Float4=@n3) {materialFunction=&BreakOutFloat4Components}
  n24 = MaterialFunctionCall!(
   Normal_Map
   Caculated_Bounds=@n21
   Pre_Skinned_Normals=@n22
   Skew Amount=@n18
   Is Vehicle=@n23.R
   Uvs_Offset=@n5
   Uvs_TileRate=@n13
   Alternate Axies contrast=@n14
  ) {materialFunction=&/MF_Projection_Skew_Normal}
  n25 = TextureObjectParameter() {
   desc=best to use a Blured version of you BaseTexture
   group=Function_FlowMap
   parameterName="FFlow__Text_BaseColor_TriArea A=Roughness"
   texture=/Game/Weapons/WeaponSkins/Wraps/Textures/UVCheaker
  }
  n26 = TextureSample(UVs=@n24.UVs, Tex=@n25) {texture=/Engine/EngineResources/DefaultTexture}
  n27 = TextureObjectParameter() {
   group=Function_FlowMap
   parameterName=FFlow__FlowMap__Text
   samplerType=SAMPLERTYPE_LinearColor
   texture=/Game/Weapons/WeaponSkins/Wraps/Textures/VanGogh/T_Wrap_VanGogh_N
  }
  n28 = ComponentMask(None=TextureSample(UVs=@n24.UVs, Tex=@n27).RGB) {g=true, r=true}
  n29 = Multiply(
   A=Multiply(A=Add(A=@n28) {constB=-0.5}) {constB=2}
   B=ScalarParameter() {defaultValue=0.1, group=Function_FlowMap, parameterName=FFlow_FlowMapEffect}
  )
  n30 = Multiply(A=Time(), B=ScalarParameter() {defaultValue=0.5, group=Function_FlowMap, parameterName=FFlow_FlowMap_Speed}) {constB=0.5}
  n31 = Frac(None=@n30)
  n32 = Multiply(A=@n29, B=@n31)
  n33 = Constant()
  n34 = TextureObjectParameter() {
   group=Function_FlowMap
   parameterName=FFlow__Text_Normal
   samplerType=SAMPLERTYPE_Normal
   texture=&FlatNormalBlue
  }
  n35 = MaterialFunctionCall(
   Normal_Map=@n34
   Caculated_Bounds=@n21
   Pre_Skinned_Normals=@n22
   Skew Amount=@n18
   Is Vehicle=@n23.R
   Uvs_Offset=Add(A=AppendVector(A=@n32, B=@n33), B=@n5)
   Uvs_TileRate=@n13
   Alternate Axies contrast=@n14
  ) {materialFunction=&MF_Projection_Skew_Normal}
  n36 = TextureObjectParameter() {group=Function_FlowMap, parameterName="FFlow__Text_BaseColor  A=Roughness", texture=&UVCheaker}
  n37 = TextureSample(UVs=@n35.UVs, Tex=@n36) {samplerType=SAMPLERTYPE_LinearColor, texture=&T_Wrap_VanGogh_N}
  n38 = Multiply(A=@n29, B=Frac(None=Add(A=@n30) {constB=0.5}))
  n39 = MaterialFunctionCall(
   Normal_Map=@n34
   Caculated_Bounds=@n21
   Pre_Skinned_Normals=@n22
   Skew Amount=@n18
   Is Vehicle=@n23.R
   Uvs_Offset=Add(A=AppendVector(A=@n38, B=@n33), B=@n5)
   Uvs_TileRate=@n13
   Alternate Axies contrast=@n14
  ) {materialFunction=&MF_Projection_Skew_Normal}
  n40 = TextureSample(UVs=@n39.UVs, Tex=@n36) {samplerType=SAMPLERTYPE_LinearColor, texture=&T_Wrap_VanGogh_N}
  n41 = Abs(None=ConstantBiasScale(None=@n31) {bias=-0.5, scale=2})
  n42 = LinearInterpolate(A=@n37.A, B=@n40.A, Alpha=@n41)
  n43 = LinearInterpolate(B=Saturate(None=LinearInterpolate(A=@n35.Raw_Normal, B=@n39.Raw_Normal, Alpha=@n41)), Alpha=@n24.Masking Alternat Axies)
  n44 = TextureSample() {samplerType=SAMPLERTYPE_Normal, texture=/Game/Weapons/FORT_Rifles/T_SCAR_N}
  n45 = GetMaterialAttributes(None=FunctionInput(Preview=SetMaterialAttributes!(MaterialAttributes, Normal=@n44.RGB) {attributeSetTypes [#1]}) {inputName=Material, inputType=FunctionInput_MaterialAttributes, sortPriority=-20}) {attributeGetTypes [#1]}
  n46 = SetMaterialAttributes!(
   MaterialAttributes=@n45.MaterialAttributes
   BaseColor=LinearInterpolate(A=@n26.RGB, B=LinearInterpolate(A=@n37.RGB, B=@n40.RGB, Alpha=@n41), Alpha=@n24.Masking Alternat Axies)
   Metallic=ScalarParameter() {group=Function_FlowMap, parameterName=FFlow_BS_Metalic, sortPriority=1}
   Specular=ScalarParameter() {defaultValue=0.5, group=Function_FlowMap, parameterName=FFlow_BS_Specular, sortPriority=1}
   Roughness=LinearInterpolate(
    A=ScalarParameter() {defaultValue=0.15, group=Function_FlowMap, parameterName=FFlow_BS_Roughness, sortPriority=1}
    B=ScalarParameter() {defaultValue=1, group=Function_FlowMap, parameterName=FFlow_BS_Roughness_High, sortPriority=1}
    Alpha=LinearInterpolate(A=@n26.A, B=@n42, Alpha=@n24.Masking Alternat Axies)
   )
   Normal=LinearInterpolate(
    A=@n45.Normal
    B=Add(A=@n45.Normal, B=@n43)
    Alpha=ScalarParameter() {defaultValue=1, group=Function_FlowMap, parameterName=FFlow_BS_Normal_Strength, sortPriority=1}
   )
  ) {attributeSetTypes [
    #2
    #3
    #4
    #5
    #1
   ]}
  n47 = MaterialFunctionCall(Color=@n3) {materialFunction=&SplitComponents}
  n48 = ScalarParameter() {desc=Side Projection Rotation, group=Function_FlowMap, parameterName=FFlow_UV_Rotation_Vehicle}
  n49 = ScalarParameter() {defaultValue=0.25, desc=Side Projection Rotation, group=Function_FlowMap, parameterName=FFlow_UV_Rotation_Mech}
  n50 = VectorParameter() {desc=Side_Projection TileRate, group=Function_FlowMap, parameterName=FFlow_UV_TileRateS_Weapon, defaultValue=<rgba 1 1 1 1>}
  n51 = VectorParameter() {desc=Side_Projection TileRate, group=Function_FlowMap, parameterName=FFlow_UV_TileRateS_Vehicle, defaultValue=<rgba 1 1 1 1>}
  n52 = Multiply(A=@n51.RGB, B=@n51.RGB)
  n53 = VectorParameter() {desc=Side_Projection TileRate, group=Function_FlowMap, parameterName=FFlow_UV_TileRateS_VehicleSide, defaultValue=<rgba 1 1 1 1>}
  n54 = Multiply(A=@n53.RGB, B=@n53.RGB)
  n55 = MaterialFunctionCall(Color=@n3) {materialFunction=&SplitComponents}
  n56 = VectorParameter() {desc=Side_Projection TileRate, group=Function_FlowMap, parameterName=FFlow_UV_TileRateS_Mech, defaultValue=<rgba 1 1 1 1>}
  n57 = LinearInterpolate(
   A=LinearInterpolate(A=Multiply(A=@n50.RGB, B=@n50.RGB), B=LinearInterpolate(A=@n52, B=@n54, Alpha=@n55.G), Alpha=@n55.R)
   B=Multiply(A=@n56.RGB, B=@n56.RGB)
   Alpha=@n55.B
  )
  n58 = MaterialFunctionCall!(
   For_Safe_Zones
   Normal_Map
   Caculated_Bounds=@n21
   Pre_Skinned_Normals=@n22
   Is Vehicle=@n47.R
   Is VehicleSide=@n47.R
   Is Mech=@n47.R
   Vehicle_Rotation=@n48
   UV_Rotation Bypass
   Mech_Rotation=@n49
   Uvs_Offset=@n5
   Offset-after rotation
   Uvs_TileRate=@n57
  ) {materialFunction=&/MF_Projection_Side_Normal}
  n59 = TextureSample(UVs=@n58.UVs, Tex=@n25) {texture=&DefaultTexture}
  n60 = MaterialFunctionCall(Color=@n3) {materialFunction=&SplitComponents}
  n61 = ComponentMask(None=TextureSample(UVs=@n58.UVs, Tex=@n27).RGB) {g=true, r=true}
  n62 = Multiply(
   A=Multiply(A=Add(A=@n61) {constB=-0.5}) {constB=2}
   B=ScalarParameter() {defaultValue=0.1, group=Function_FlowMap, parameterName=FFlow_FlowMapEffect}
  )
  n63 = Multiply(A=Time(), B=ScalarParameter() {defaultValue=0.5, group=Function_FlowMap, parameterName=FFlow_FlowMap_Speed}) {constB=0.5}
  n64 = Frac(None=@n63)
  n65 = Multiply(A=@n62, B=@n64)
  n66 = Constant()
  n67 = MaterialFunctionCall(
   Normal_Map=@n34
   Caculated_Bounds=@n21
   Pre_Skinned_Normals=@n22
   Is Vehicle=@n60.R
   Is VehicleSide=@n60.R
   Is Mech=@n60.R
   Vehicle_Rotation=@n48
   Mech_Rotation=@n49
   Uvs_Offset=Add(A=AppendVector(A=@n65, B=@n66), B=@n5)
   Uvs_TileRate=@n57
  ) {materialFunction=&MF_Projection_Side_Normal}
  n68 = TextureSample(UVs=@n67.UVs, Tex=@n36) {samplerType=SAMPLERTYPE_LinearColor, texture=&T_Wrap_VanGogh_N}
  n69 = Multiply(A=@n62, B=Frac(None=Add(A=@n63) {constB=0.5}))
  n70 = MaterialFunctionCall(
   Normal_Map=@n34
   Caculated_Bounds=@n21
   Pre_Skinned_Normals=@n22
   Is Vehicle=@n60.R
   Is VehicleSide=@n60.R
   Is Mech=@n60.R
   Vehicle_Rotation=@n48
   Mech_Rotation=@n49
   Uvs_Offset=Add(A=AppendVector(A=@n69, B=@n66), B=@n5)
   Uvs_TileRate=@n57
  ) {materialFunction=&MF_Projection_Side_Normal}
  n71 = TextureSample(UVs=@n70.UVs, Tex=@n36) {samplerType=SAMPLERTYPE_LinearColor, texture=&T_Wrap_VanGogh_N}
  n72 = Abs(None=ConstantBiasScale(None=@n64) {bias=-0.5, scale=2})
  n73 = MaterialFunctionCall(Color=Abs(None=@n22)) {materialFunction=&SplitComponents}
  n74 = MaterialFunctionCall(Color=@n3) {materialFunction=&SplitComponents}
  n75 = LinearInterpolate(A=@n73.B, B=@n73.B, Alpha=@n74.G)
  n76 = MaterialFunctionCall(In=LinearInterpolate(A=LinearInterpolate(A=@n73.R, B=@n75, Alpha=@n74.R), B=@n73.G, Alpha=@n74.B), Contrast=@n14) {materialFunction=&CheapContrast}
  n77 = LinearInterpolate(A=@n68.A, B=@n71.A, Alpha=@n72)
  n78 = LinearInterpolate(A=Constant3Vector(), B=Saturate(None=LinearInterpolate(A=@n67.Raw_Normal, B=@n70.Raw_Normal, Alpha=@n72)), Alpha=@n76.Result)
  n79 = SetMaterialAttributes(
   MaterialAttributes=@n45.MaterialAttributes
   BaseColor=LinearInterpolate(A=@n59.RGB, B=LinearInterpolate(A=@n68.RGB, B=@n71.RGB, Alpha=@n72), Alpha=@n76.Result)
   Metallic=ScalarParameter() {group=Function_FlowMap, parameterName=FFlow_BS_Metalic, sortPriority=1}
   Specular=ScalarParameter() {defaultValue=0.5, group=Function_FlowMap, parameterName=FFlow_BS_Specular, sortPriority=1}
   Roughness=LinearInterpolate(
    A=ScalarParameter() {defaultValue=0.15, group=Function_FlowMap, parameterName=FFlow_BS_Roughness, sortPriority=1}
    B=ScalarParameter() {defaultValue=1, group=Function_FlowMap, parameterName=FFlow_BS_Roughness_High, sortPriority=1}
    Alpha=LinearInterpolate(A=@n59.A, B=@n77, Alpha=@n76.Result)
   )
   Normal=LinearInterpolate(
    A=@n45.Normal
    B=Add(A=@n45.Normal, B=@n78)
    Alpha=ScalarParameter() {defaultValue=1, group=Function_FlowMap, parameterName=FFlow_BS_Normal_Strength, sortPriority=1}
   )
  ) {attributeSetTypes [
    #2
    #3
    #4
    #5
    #1
   ]}
  n80 = MaterialFunctionCall(Color=@n3) {materialFunction=&SplitComponents}
  n81 = LinearInterpolate(Alpha=@n80.R)
  n82 = StaticSwitchParameter(
   True=BlendMaterialAttributes(A=@n46, B=@n79, Alpha=LinearInterpolate(A=@n81, Alpha=@n80.B)) {vertexAttributeBlendType=UseA}
   False=StaticSwitchParameter(True=@n46, False=@n79) {group=Function_FlowMap, parameterName=FFlow_Projection_Skew, sortPriority=0}
  ) {group=Function_FlowMap, parameterName="FFlow_Projection_Split Vehicles=Side Weapons=Skew", sortPriority=0}
  n83 = FunctionInput() {bUsePreviewValueAsDefault=true, inputName="Customization Mask (_CM B)", inputType=FunctionInput_Scalar, sortPriority=24}
  n84 = LinearInterpolate(
   A=@n83
   B=FunctionInput() {bUsePreviewValueAsDefault=true, inputName="Customization Mask (_CM A+B)", inputType=FunctionInput_Scalar, sortPriority=24}
   Alpha=ScalarParameter() {group=Function_FlowMap, parameterName="FFlow_Customization Mask 0=Primary--> 1=Most of the area", sortPriority=-10}
  )
  n85 = ScalarParameter() {group=Function_FlowMap, parameterName=FFlow_Customization Mask Remove, sortPriority=-10}
  n86 = NamedRerouteUsage()
  n87 = NamedRerouteUsage()
  n88 = LinearInterpolate(A=Saturate(None=Add(A=@n84, B=@n85)), B=@n83, Alpha=Saturate(None=Subtract(A=@n86, B=@n87)))
  FunctionOutput(None=StaticSwitchParameter(True=BlendMaterialAttributes(A=@n45.MaterialAttributes, B=@n82, Alpha=@n88) {vertexAttributeBlendType=UseA}, False=@n45.MaterialAttributes) {desc="side or Skew projection using FlowMaps. ", group=Customization, parameterName=Function_Flowmap})
  NamedRerouteDeclaration(None=@n23.A) {name=IsCharacter, nodeColor=<rgba 0.8634 1 0 1>}
  NamedRerouteDeclaration(None=FunctionInput() {bUsePreviewValueAsDefault=true, inputName=isCharAccent, inputType=FunctionInput_Scalar, sortPriority=78}) {name=isCharAccent, nodeColor=<rgba 0 0.9253 1 1>}
 ]}
