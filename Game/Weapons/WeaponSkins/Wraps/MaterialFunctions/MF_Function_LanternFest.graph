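# Material function for the LanternFest wrap skin, applied through the customization mask behind a static switch, with IsState choosing weapon, vehicle, mech or character tuning.
# Three camera-offset parallax texture layers are projected from normalized pre-skinned bounds using skew or side projection, then brightened by curvature.
# A lantern mask shapes pulsing sphere glows and warped, radially scrolling dots over those layers.
# The result drives emissive, scaled by time of day.
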MaterialFunction /Game/Weapons/WeaponSkins/Wraps/MaterialFunctions/MF_Function_LanternFest
graph {nodes [
  Constant()
  Constant()
  n1 = FunctionInput(Preview=SetMaterialAttributes!(MaterialAttributes, Normal=TextureSample() {samplerType=SAMPLERTYPE_Normal, texture=/Game/Weapons/FORT_Rifles/T_SCAR_N}.RGB) {attributeSetTypes [#1]}) {bUsePreviewValueAsDefault=true, inputName=Material, inputType=FunctionInput_MaterialAttributes, sortPriority=-20}
  n2 = GetMaterialAttributes(None=@n1)
  n3 = VectorParameter() {group=Function_Parallax, parameterName=FParallax_BaseColor, sortPriority=0, defaultValue=<rgba 0 0 0 1>}
  n4 = ScalarParameter() {defaultValue=2, group=Function_Parallax, parameterName=FParallax_Edge_Strength, sortPriority=0}
  n5 = VectorParameter() {group=Function_Parallax, parameterName=FParallax_Tri_Mask_Color, sortPriority=0, defaultValue=<rgba 0 0 0 1>}
  n6 = VertexInterpolator(VS=PreSkinnedPosition())
  n7 = PreSkinnedLocalBounds()
  n8 = FunctionInput(Preview=Divide(A=Subtract(A=@n6.PS, B=@n7.Min), B=Subtract(A=@n7.Max, B=@n7.Max))) {bUsePreviewValueAsDefault=true, inputName=Calcated_Bounds, sortPriority=11}
  n9 = Add(A=FunctionInput(Preview=TextureSample() {samplerType=SAMPLERTYPE_Masks, texture=/Game/Weapons/FORT_Rifles/Textures/T_Scar_CM}.G) {bUsePreviewValueAsDefault=true, inputName=Curavture, inputType=FunctionInput_Scalar, sortPriority=24}) {constB=-0.5}
  n10 = Add(A=@n8, B=Multiply(A=@n9, B=ScalarParameter() {group=Function_Parallax, parameterName=FParallax_Cuvatur_Warp_amount, sortPriority=0}))
  n11 = Transform(None=CameraVectorWS()) {transformSourceType=TRANSFORMSOURCE_World, transformType=TRANSFORM_Local}
  n12 = Saturate(None=MaterialFunctionCall() {materialFunction=/Engine/Functions/Engine_MaterialFunctions02/WorldPositionOffset/ObjectScale}.Scale Y)
  n13 = FunctionInput(Preview=Constant()) {bUsePreviewValueAsDefault=true, inputName=IsState, inputType=FunctionInput_Vector4}
  n14 = MaterialFunctionCall(Color=@n13) {materialFunction=&SplitComponents}
  n15 = LinearInterpolate(
   A=LinearInterpolate(
    A=ScalarParameter() {defaultValue=-2, group=Function_Parallax, parameterName=FParallax_Depth_Amount_Layer_0_Weapon}
    B=ScalarParameter() {group=Function_Parallax, parameterName=FParallax_Depth_Amount_Layer_0_Vehicle}
    Alpha=@n14.R
   )
   B=ScalarParameter() {group=Function_Parallax, parameterName=FParallax_Depth_Amount_Layer_0_Mech}
   Alpha=@n14.B
  )
  n16 = Add(A=@n10, B=Multiply(A=@n11, B=Multiply(A=@n12, B=@n15)))
  n17 = ScalarParameter() {defaultValue=0.5, group=Function_Parallax, parameterName=FParallax_Skew_Amount, sortPriority=0}
  n18 = MaterialFunctionCall(Color=@n13) {materialFunction=&SplitComponents}
  n19 = VectorParameter() {group=Function_Parallax, parameterName=FParallaxUV_TileRate_Layer_0_Weapon, defaultValue=<rgba 1 2 0 1>}
  n20 = VectorParameter() {group=Function_Parallax, parameterName=FParallaxUV_TileRate_Layer_0_Vehicle, defaultValue=<rgba 1 2 0 1>}
  n21 = Multiply(A=@n20.RGB, B=@n20.RGB)
  n22 = VectorParameter() {group=Function_Parallax, parameterName=FParallaxUV_TileRate_Layer_0_VehicleSide, defaultValue=<rgba 1 2 0 1>}
  n23 = Multiply(A=@n22.RGB, B=@n22.RGB)
  n24 = MaterialFunctionCall(Float4=@n13) {materialFunction=&BreakOutFloat4Components}
  n25 = VectorParameter() {group=Function_Parallax, parameterName=FParallaxUV_TileRate_Layer_0_Mech, defaultValue=<rgba 1 2 0 1>}
  n26 = VectorParameter() {group=Function_Parallax, parameterName=FParallaxUV_TileRate_Layer_0_Character, defaultValue=<rgba 1 1 0 1>}
  n27 = LinearInterpolate(
   A=LinearInterpolate(A=Multiply(A=@n19.RGB, B=@n19.RGB), B=LinearInterpolate(A=@n21, B=@n23, Alpha=@n24.G), Alpha=@n24.R)
   B=LinearInterpolate(A=Multiply(A=@n25.RGB, B=@n25.RGB), B=Multiply(A=@n26.RGB, B=@n26.RGB), Alpha=@n24.A)
   Alpha=@n24.B
  )
  n28 = ScalarParameter() {defaultValue=1, group=Function_Parallax, parameterName=FParallax_Skew_Alternate_Axies_Contrast, sortPriority=0}
  n29 = FunctionInput(Preview=VertexInterpolator(VS=PreSkinnedNormal()).PS) {bUsePreviewValueAsDefault=true, inputName=Pre-Skinned Local Normal, sortPriority=10}
  n30 = ScalarParameter() {group=Function_Parallax, parameterName=FParallax_UV_Rotation_Layer_0_Vehicle}
  n31 = ScalarParameter() {group=Function_Parallax, parameterName=FParallax_UV_Rotation_Layer_0_VehicleSide}
  n32 = MaterialFunctionCall(Color=@n13) {materialFunction=&SplitComponents}
  n33 = LinearInterpolate(
   A=LinearInterpolate(
    A=ScalarParameter() {group=Function_Parallax, parameterName=FParallax_UV_Rotation_Layer_0_Weapon}
    B=LinearInterpolate(A=@n30, B=@n31, Alpha=@n32.G)
    Alpha=@n32.R
   )
   B=ScalarParameter() {defaultValue=0.25, group=Function_Parallax, parameterName=FParallax_UV_Rotation_Layer_0_Mech}
   Alpha=@n32.B
  )
  n34 = Time()
  n35 = Multiply(A=VectorParameter() {group=Function_Parallax, parameterName=FParallax_UV_Speed 0, defaultValue=<rgba 0 0 0 1>}.RGB, B=@n34)
  n36 = Add(
   A=MaterialFunctionCall(
    UVs=MaterialFunctionCall!(
     Normal_Map
     Caculated_Bounds=@n16
     Pre_Skinned_Normals=@n29
     Skew Amount=@n17
     Is Vehicle=@n18.R
     Uvs_Offset
     Uvs_TileRate=@n27
     Alternate Axies contrast=@n28
    ) {materialFunction=&/MF_Projection_Skew_Normal}.UVs
    "Rotation Angle (0-1)"=@n33
   ) {materialFunction=&CustomRotator}.Rotated Values
   B=ComponentMask(None=@n35) {g=true, r=true}
  )
  n37 = MaterialFunctionCall(Color=@n13) {materialFunction=&SplitComponents}
  n38 = LinearInterpolate(
   A=LinearInterpolate(
    A=ScalarParameter() {defaultValue=-0.5, group=Function_Parallax, parameterName=FParallax_Depth_Amount_Layer_2_Weapon}
    B=ScalarParameter() {group=Function_Parallax, parameterName=FParallax_Depth_Amount_Layer_2_Vehicle}
    Alpha=@n37.R
   )
   B=ScalarParameter() {group=Function_Parallax, parameterName=FParallax_Depth_Amount_Layer_2_Mech}
   Alpha=@n37.B
  )
  n39 = Add(A=@n10, B=Multiply(A=@n11, B=Multiply(A=@n12, B=@n38)))
  n40 = VectorParameter() {group=Function_Parallax, parameterName=FParallaxUV_TileRate_Layer_2_Weapon, defaultValue=<rgba 1 2 0 1>}
  n41 = VectorParameter() {group=Function_Parallax, parameterName=FParallaxUV_TileRate_Layer_2_Vehicle, defaultValue=<rgba 1 2 0 1>}
  n42 = Multiply(A=@n41.RGB, B=@n41.RGB)
  n43 = VectorParameter() {group=Function_Parallax, parameterName=FParallaxUV_TileRate_Layer_2_VehicleSide, defaultValue=<rgba 1 2 0 1>}
  n44 = Multiply(A=@n43.RGB, B=@n43.RGB)
  n45 = MaterialFunctionCall(Float4=@n13) {materialFunction=&BreakOutFloat4Components}
  n46 = VectorParameter() {group=Function_Parallax, parameterName=FParallaxUV_TileRate_Layer_2_Mech, defaultValue=<rgba 1 2 0 1>}
  n47 = VectorParameter() {group=Function_Parallax, parameterName=FParallaxUV_TileRate_Layer_2_Character, defaultValue=<rgba 1 1 0 1>}
  n48 = LinearInterpolate(
   A=LinearInterpolate(A=Multiply(A=@n40.RGB, B=@n40.RGB), B=LinearInterpolate(A=@n42, B=@n44, Alpha=@n45.G), Alpha=@n45.R)
   B=LinearInterpolate(A=Multiply(A=@n46.RGB, B=@n46.RGB), B=Multiply(A=@n47.RGB, B=@n47.RGB), Alpha=@n45.A)
   Alpha=@n45.B
  )
  n49 = MaterialFunctionCall(Caculated_Bounds=@n39, Pre_Skinned_Normals=@n29, Skew Amount=@n17, Is Vehicle=@n18.R, Uvs_TileRate=@n48, Alternate Axies contrast=@n28) {materialFunction=&MF_Projection_Skew_Normal}
  n50 = ScalarParameter() {group=Function_Parallax, parameterName=FParallax_UV_Rotation_Layer_2_Vehicle}
  n51 = ScalarParameter() {group=Function_Parallax, parameterName=FParallax_UV_Rotation_Layer_2_VehicleSide}
  n52 = MaterialFunctionCall(Color=@n13) {materialFunction=&SplitComponents}
  n53 = LinearInterpolate(
   A=LinearInterpolate(
    A=ScalarParameter() {group=Function_Parallax, parameterName=FParallax_UV_Rotation_Layer_2_Weapon}
    B=LinearInterpolate(A=@n50, B=@n51, Alpha=@n52.G)
    Alpha=@n52.R
   )
   B=ScalarParameter() {defaultValue=0.25, group=Function_Parallax, parameterName=FParallax_UV_Rotation_Layer_2_Mech}
   Alpha=@n52.B
  )
  n54 = Multiply(A=VectorParameter() {group=Function_Parallax, parameterName=FParallax_UV_Speed 2, defaultValue=<rgba 0 0 0 1>}.RGB, B=@n34)
  n55 = TextureSampleParameter2D(UVs=Add(
    A=MaterialFunctionCall(UVs=@n49.UVs, "Rotation Angle (0-1)"=@n53) {materialFunction=&CustomRotator}.Rotated Values
    B=ComponentMask(None=@n54) {g=true, r=true}
   )) {group=Function_Parallax, parameterName=FParallax_Text__Layer_2, texture=&T_BlacksRGB}
  n56 = MaterialFunctionCall(Color=@n13) {materialFunction=&SplitComponents}
  n57 = LinearInterpolate(
   A=LinearInterpolate(
    A=ScalarParameter() {defaultValue=0.15, group=Function_Parallax, parameterName=FParallax_Depth_Amount_Layer_3_Weapon}
    B=ScalarParameter() {group=Function_Parallax, parameterName=FParallax_Depth_Amount_Layer_3_Vehicle}
    Alpha=@n56.R
   )
   B=ScalarParameter() {group=Function_Parallax, parameterName=FParallax_Depth_Amount_Layer_3_Mech}
   Alpha=@n56.B
  )
  n58 = Add(A=@n10, B=Multiply(A=@n11, B=Multiply(A=@n12, B=@n57)))
  n59 = VectorParameter() {group=Function_Parallax, parameterName=FParallaxUV_TileRate_Layer_3_Weapon, defaultValue=<rgba 1 2 0 1>}
  n60 = VectorParameter() {group=Function_Parallax, parameterName=FParallaxUV_TileRate_Layer_3_Vehicle, defaultValue=<rgba 1 2 0 1>}
  n61 = Multiply(A=@n60.RGB, B=@n60.RGB)
  n62 = VectorParameter() {group=Function_Parallax, parameterName=FParallaxUV_TileRate_Layer_3_VehicleSide, defaultValue=<rgba 1 2 0 1>}
  n63 = Multiply(A=@n62.RGB, B=@n62.RGB)
  n64 = MaterialFunctionCall(Float4=@n13) {materialFunction=&BreakOutFloat4Components}
  n65 = VectorParameter() {group=Function_Parallax, parameterName=FParallaxUV_TileRate_Layer_3_Mech, defaultValue=<rgba 1 2 0 1>}
  n66 = VectorParameter() {group=Function_Parallax, parameterName=FParallaxUV_TileRate_Layer_3_Character, defaultValue=<rgba 1 1 0 1>}
  n67 = LinearInterpolate(
   A=LinearInterpolate(A=Multiply(A=@n59.RGB, B=@n59.RGB), B=LinearInterpolate(A=@n61, B=@n63, Alpha=@n64.G), Alpha=@n64.R)
   B=LinearInterpolate(A=Multiply(A=@n65.RGB, B=@n65.RGB), B=Multiply(A=@n66.RGB, B=@n66.RGB), Alpha=@n64.A)
   Alpha=@n64.B
  )
  n68 = MaterialFunctionCall(Caculated_Bounds=@n58, Pre_Skinned_Normals=@n29, Skew Amount=@n17, Is Vehicle=@n18.R, Uvs_TileRate=@n67, Alternate Axies contrast=@n28) {materialFunction=&MF_Projection_Skew_Normal}
  n69 = ScalarParameter() {group=Function_Parallax, parameterName=FParallax_UV_Rotation_Layer_3_Vehicle}
  n70 = ScalarParameter() {group=Function_Parallax, parameterName=FParallax_UV_Rotation_Layer_3_VehicleSide}
  n71 = MaterialFunctionCall(Color=@n13) {materialFunction=&SplitComponents}
  n72 = LinearInterpolate(
   A=LinearInterpolate(
    A=ScalarParameter() {group=Function_Parallax, parameterName=FParallax_UV_Rotation_Layer_3_Weapon}
    B=LinearInterpolate(A=@n69, B=@n70, Alpha=@n71.G)
    Alpha=@n71.R
   )
   B=ScalarParameter() {defaultValue=0.25, group=Function_Parallax, parameterName=FParallax_UV_Rotation_Layer_3_Mech}
   Alpha=@n71.B
  )
  n73 = Multiply(A=VectorParameter() {group=Function_Parallax, parameterName=FParallax_UV_Speed 3, defaultValue=<rgba 0 0 0 1>}.RGB, B=@n34)
  n74 = TextureSampleParameter2D(UVs=Add(
    A=MaterialFunctionCall(UVs=@n68.UVs, "Rotation Angle (0-1)"=@n72) {materialFunction=&CustomRotator}.Rotated Values
    B=ComponentMask(None=@n73) {g=true, r=true}
   )) {group=Function_Parallax, parameterName=FParallax_Text__Layer_3, texture=&T_BlacksRGB}
  n75 = LinearInterpolate(
   A=LinearInterpolate(
    A=TextureSampleParameter2D(UVs=@n36) {group=Function_Parallax, parameterName=FParallax_Text__Layer_0, texture=&T_BlacksRGB}.RGB
    B=@n55.RGB
    Alpha=@n55.RGB
   )
   B=@n74.RGB
   Alpha=@n74.RGB
  )
  n76 = MaterialFunctionCall(Color=@n13) {materialFunction=&SplitComponents}
  n77 = TextureSampleParameter2D(UVs=MaterialFunctionCall!(
    For_Safe_Zones
    Normal_Map
    Caculated_Bounds=@n16
    Pre_Skinned_Normals=@n29
    Is Vehicle=@n76.R
    Is VehicleSide=@n76.R
    Is Mech=@n76.R
    Vehicle_Rotation
    UV_Rotation Bypass=@n33
    Mech_Rotation
    Uvs_Offset=@n35
    Offset-after rotation
    Uvs_TileRate=@n27
   ) {materialFunction=&/MF_Projection_Side_Normal}.UVs) {group=Function_Parallax, parameterName=FParallax_Text__Layer_0, texture=&T_BlacksRGB}
  n78 = TextureSampleParameter2D(UVs=MaterialFunctionCall(
    Caculated_Bounds=@n39
    Pre_Skinned_Normals=@n29
    Is Vehicle=@n76.R
    Is VehicleSide=@n76.R
    Is Mech=@n76.R
    UV_Rotation Bypass=@n53
    Uvs_Offset=@n54
    Uvs_TileRate=@n48
   ) {materialFunction=&MF_Projection_Side_Normal}.UVs) {group=Function_Parallax, parameterName=FParallax_Text__Layer_2, texture=&T_BlacksRGB}
  n79 = TextureSampleParameter2D(UVs=MaterialFunctionCall(
    Caculated_Bounds=@n58
    Pre_Skinned_Normals=@n29
    Is Vehicle=@n76.R
    Is VehicleSide=@n76.R
    Is Mech=@n76.R
    UV_Rotation Bypass=@n72
    Uvs_Offset=@n73
    Uvs_TileRate=@n67
   ) {materialFunction=&MF_Projection_Side_Normal}.UVs) {group=Function_Parallax, parameterName=FParallax_Text__Layer_3, texture=&T_BlacksRGB}
  n80 = MaterialFunctionCall(Color=Abs(None=@n29)) {materialFunction=&SplitComponents}
  n81 = MaterialFunctionCall(Color=@n13) {materialFunction=&SplitComponents}
  n82 = LinearInterpolate(A=@n80.R, B=LinearInterpolate(A=@n80.B, B=@n80.B, Alpha=@n81.G), Alpha=@n81.R)
  n83 = LinearInterpolate(
   A=@n5.RGB
   B=LinearInterpolate(A=LinearInterpolate(A=@n77.RGB, B=@n78.RGB, Alpha=@n78.RGB), B=@n79.RGB, Alpha=@n79.RGB)
   Alpha=MaterialFunctionCall(
    In=LinearInterpolate(A=@n82, B=@n80.G, Alpha=@n81.B)
    Contrast=ScalarParameter() {defaultValue=0.5, group=Function_Parallax, parameterName=FParallax_Tri_Planner_MaskContrast, sortPriority=5}
   ) {materialFunction=&CheapContrast}.Result
  )
  n84 = LinearInterpolate(
   A=StaticSwitchParameter(True=LinearInterpolate(A=@n5.RGB, B=@n75, Alpha=@n68.Masking Alternat Axies), False=@n83) {defaultValue=true, group=Function_Parallax, parameterName=FParallax_Skew_Projection, sortPriority=-1}
   B=@n83
   Alpha=NamedRerouteUsage()
  )
  n85 = Multiply(
   A=LinearInterpolate(A=@n84, B=Multiply(A=@n84, B=@n4), Alpha=Saturate(None=@n9))
   B=ScalarParameter() {defaultValue=0.75, group=Function_Parallax, parameterName=FParallax_Emissive_Amount, sortPriority=0}
  )
  n86 = MaterialFunctionCall(Float4=@n13) {materialFunction=&BreakOutFloat4Components}
  n87 = MaterialFunctionCall(
   Caculated_Bounds=@n8
   Pre_Skinned_Normals=@n29
   Is Vehicle=@n86.R
   Is VehicleSide=@n86.R
   Is Mech=@n86.R
   Vehicle_Rotation=ScalarParameter() {group=Wrap_LanternFest_UVs, parameterName=LF_Lantern_Rotation_Vehicle}
   Mech_Rotation=ScalarParameter() {defaultValue=0.25, group=Wrap_LanternFest_UVs, parameterName=LF_Lantern_Rotation_Mech -----}
  ) {materialFunction=&MF_Projection_Side_Normal}
  n88 = ScalarParameter() {defaultValue=1, group=Wrap_LanternFest_UVs, parameterName=LF_Lantern_TileRate_Vehicle}
  n89 = ScalarParameter() {defaultValue=1, group=Wrap_LanternFest_UVs, parameterName=LF_Lantern_TileRate_VehicleSide}
  n90 = MaterialFunctionCall(Float4=@n13) {materialFunction=&BreakOutFloat4Components}
  n91 = LinearInterpolate(
   A=LinearInterpolate(
    A=ScalarParameter() {defaultValue=1, group=Wrap_LanternFest_UVs, parameterName=LF_Lantern_TileRate_Weapon}
    B=LinearInterpolate(A=@n88, B=@n89, Alpha=@n90.G)
    Alpha=@n90.R
   )
   B=LinearInterpolate(
    A=ScalarParameter() {defaultValue=1, group=Wrap_LanternFest_UVs, parameterName=LF_Lantern_TileRate_Mech -----}
    B=ScalarParameter() {defaultValue=1, group=Wrap_LanternFest_UVs, parameterName=LF_Lantern_TileRate_Character}
    Alpha=@n90.A
   )
   Alpha=@n90.B
  )
  n92 = VectorParameter() {group=Wrap_LanternFest_UVs, parameterName=LF_Lantern_Offset_Vehicle, defaultValue=<rgba 0 0 0 1>}
  n93 = VectorParameter() {group=Wrap_LanternFest_UVs, parameterName=LF_Lantern_Offset_Vehicle_Side, defaultValue=<rgba 0 0 0 1>}
  n94 = LinearInterpolate(
   A=LinearInterpolate(
    A=VectorParameter() {group=Wrap_LanternFest_UVs, parameterName=LF_Lantern_Offset_Weapon, defaultValue=<rgba 0 0 0 1>}.RGB
    B=LinearInterpolate(A=@n92.RGB, B=@n93.RGB, Alpha=@n90.G)
    Alpha=@n90.R
   )
   B=LinearInterpolate(
    A=VectorParameter() {group=Wrap_LanternFest_UVs, parameterName=LF_Lantern_Offset_Mech -----, defaultValue=<rgba 0 0 0 1>}.RGB
    B=VectorParameter() {group=Wrap_LanternFest_UVs, parameterName=LF_Lantern_Offset_Character, defaultValue=<rgba 0 0 0 1>}.RGB
    Alpha=@n90.A
   )
   Alpha=@n90.B
  )
  n95 = TextureObjectParameter() {group=Wrap_LanternFest, parameterName=LF_MaskMap_01, texture=&T_BlacksRGB}
  n96 = TextureSample(UVs=Add(A=Multiply(A=@n87.UVs, B=@n91), B=ComponentMask(None=@n94) {g=true, r=true}), Tex=@n95) {samplerType=SAMPLERTYPE_Normal, texture=&T_SCAR_N}
  n97 = ScalarParameter() {defaultValue=0.5, group=Wrap_LanternFest, parameterName=LF_SphereGlow_PulseMin}
  n98 = ScalarParameter() {defaultValue=1, group=Wrap_LanternFest, parameterName=LF_SphereGlow_PulseMax}
  n99 = Multiply(A=Time(), B=ScalarParameter() {defaultValue=0.5, group=Wrap_LanternFest, parameterName=LF_SphereGlow_PulseTime}) {constB=0.1}
  n100 = Divide(A=Add(A=Sine(None=@n99) {period=0.5}))
  n101 = VectorParameter() {group=Wrap_LanternFest, parameterName=LF_Sphere_Color, sortPriority=0, defaultValue=<rgba 0.7887 0.4242 0.0183 1>}
  n102 = Multiply(
   A=Multiply(
    A=LinearInterpolate(A=@n97, B=@n98, Alpha=@n100)
    B=ScalarParameter() {defaultValue=5, group=Wrap_LanternFest, parameterName=LF_SphereGlow_Amount}
   )
   B=@n101.RGB
  )
  n103 = Constant() {r=0.5}
  n104 = Multiply(A=Time(), B=ScalarParameter() {defaultValue=0.125, group=Wrap_LanternFest, parameterName=LF_Sphere_Radius_PulseTime}) {constB=0.1}
  n105 = Divide(A=Add(A=Sine(None=@n104) {period=0.5}))
  n106 = ScalarParameter() {defaultValue=0.8, group=Wrap_LanternFest, parameterName=LF_Sphere_Radius_Vehicle_Min}
  n107 = ScalarParameter() {defaultValue=1, group=Wrap_LanternFest, parameterName=LF_Sphere_Radius_Vehicle_Max}
  n108 = ScalarParameter() {defaultValue=0.8, group=Wrap_LanternFest, parameterName=LF_Sphere_Radius_VehicleSide_Min}
  n109 = ScalarParameter() {defaultValue=1, group=Wrap_LanternFest, parameterName=LF_Sphere_Radius_VehicleSide_Max}
  n110 = MaterialFunctionCall(Color=@n13) {materialFunction=&SplitComponents}
  n111 = LinearInterpolate(
   A=LinearInterpolate(
    A=ScalarParameter() {defaultValue=0.8, group=Wrap_LanternFest, parameterName=LF_Sphere_Radius_Weapon_Min}
    B=ScalarParameter() {defaultValue=1, group=Wrap_LanternFest, parameterName=LF_Sphere_Radius_Weapon_Max}
    Alpha=@n105
   )
   B=LinearInterpolate(A=LinearInterpolate(A=@n106, B=@n107, Alpha=@n105), B=LinearInterpolate(A=@n108, B=@n109, Alpha=@n105), Alpha=@n110.G)
   Alpha=@n110.R
  )
  n112 = LinearInterpolate(
   A=ScalarParameter() {defaultValue=0.8, group=Wrap_LanternFest, parameterName=LF_Sphere_Radius_Mech_Min}
   B=ScalarParameter() {defaultValue=1, group=Wrap_LanternFest, parameterName=LF_Sphere_Radius_Mech_Max}
   Alpha=@n105
  )
  n113 = VectorParameter() {group=Wrap_LanternFest_UVs, parameterName=LF_Sphere_Offset_Weapon, defaultValue=<rgba -0.46 0 0 1>}
  n114 = VectorParameter() {group=Wrap_LanternFest_UVs, parameterName=LF_Sphere_Offset_Vehicle, defaultValue=<rgba -0.46 0 0 1>}
  n115 = MaterialFunctionCall(
   Caculated_Bounds=@n8
   Pre_Skinned_Normals=@n29
   Is Vehicle=@n86.R
   Is VehicleSide=@n86.R
   Is Mech=@n86.R
   Vehicle_Rotation=ScalarParameter() {group=Wrap_LanternFest_UVs, parameterName=LF_Rotation_Vehicle, sortPriority=-10}
   Mech_Rotation=ScalarParameter() {defaultValue=0.25, group=Wrap_LanternFest_UVs, parameterName=LF_Rotation_Mech -----, sortPriority=-10}
   Uvs_Offset=LinearInterpolate(
    A=LinearInterpolate(A=@n113.RGB, B=@n114.RGB, Alpha=@n86.R)
    B=VectorParameter() {group=Wrap_LanternFest_UVs, parameterName=LF_Sphere_Offset_Character, defaultValue=<rgba -0.46 0 0 1>}.RGB
    Alpha=@n86.A
   )
   Uvs_TileRate=Constant() {r=1}
  ) {materialFunction=&MF_Projection_Side_Normal}
  n116 = SphereMask(
   A=@n115.UVs
   B=@n103
   Radius=Saturate(None=LinearInterpolate(A=@n111, B=@n112, Alpha=@n110.B))
   Hardness=ScalarParameter() {group=Wrap_LanternFest, parameterName=LF_Sphere_Hardness}
  ) {hardnessPercent=0}
  n117 = MaterialFunctionCall(Color=Abs(None=@n29)) {materialFunction=&SplitComponents}
  n118 = MaterialFunctionCall(Color=@n13) {materialFunction=&SplitComponents}
  n119 = LinearInterpolate(A=@n117.B, B=@n117.B, Alpha=@n118.G)
  n120 = MaterialFunctionCall(In=LinearInterpolate(A=LinearInterpolate(A=@n117.R, B=@n119, Alpha=@n118.R), B=@n117.G, Alpha=@n118.B)) {materialFunction=&CheapContrast}
  n121 = ScalarParameter() {defaultValue=10, group=Wrap_LanternFest, parameterName=LF_Dots_Intensity-----}
  n122 = VectorParameter() {group=Wrap_LanternFest_UVs, parameterName=LF_Dots_Offset_Vehicle, defaultValue=<rgba 0 0 0 1>}
  n123 = VectorParameter() {group=Wrap_LanternFest_UVs, parameterName=LF_Dots_Offset_Vehicle_Side, defaultValue=<rgba 0 0 0 1>}
  n124 = MaterialFunctionCall(Color=@n13) {materialFunction=&SplitComponents}
  n125 = LinearInterpolate(
   A=LinearInterpolate(
    A=VectorParameter() {group=Wrap_LanternFest_UVs, parameterName=LF_Dots_Offset_Weapon, defaultValue=<rgba 0 0 0 1>}.RGB
    B=LinearInterpolate(A=@n122.RGB, B=@n123.RGB, Alpha=@n124.G)
    Alpha=@n124.R
   )
   B=VectorParameter() {group=Wrap_LanternFest_UVs, parameterName=LF_Dots_Offset_Mech -----, defaultValue=<rgba 0 0 0 1>}.RGB
   Alpha=@n124.B
  )
  n126 = MaterialFunctionCall(Vector or UVs=Add(A=@n115.UVs No Offsets/Tiling, B=ComponentMask(None=@n125) {g=true, r=true})) {materialFunction=&VectorToRadialValue}
  n127 = Time()
  n128 = ComponentMask(None=VectorParameter() {group=Wrap_LanternFest, parameterName=LF_Dots_MaskSpeed, defaultValue=<rgba -0.03 -0.04 0 1>}.RGB) {g=true, r=true}
  n129 = Multiply(
   A=Add(A=@n126.Radial Coordinates, B=Multiply(A=@n127, B=@n128))
   B=ScalarParameter() {defaultValue=2, group=Wrap_LanternFest, parameterName=LF_Dots_WarpGradientTileRate}
  )
  n130 = Multiply(A=@n115.UVs No Offsets/Tiling, B=ScalarParameter() {defaultValue=0.2, group=Wrap_LanternFest_UVs, parameterName=LF_Sphere_Warp_TileRate})
  n131 = VectorParameter() {group=Wrap_LanternFest_UVs, parameterName=LF_Sphere_WarpGradientspeed, defaultValue=<rgba -0.025 -0.1 0 1>}
  n132 = Multiply(A=Time(), B=ComponentMask(None=@n131.RGB) {g=true, r=true})
  n133 = ChannelMaskParameter(None=TextureSample(UVs=Add(A=@n130, B=@n132), Tex=@n95) {samplerType=SAMPLERTYPE_LinearGrayscale}.RGBA) {group=Wrap_LanternFest_UVs, maskChannel=Blue, parameterName=LF_Sphere_Warp_Channel, defaultValue=<rgba 0 0 1 0>}
  n134 = Multiply(A=@n133.R, B=ScalarParameter() {defaultValue=0.1, group=Wrap_LanternFest_UVs, parameterName=LF_Sphere_Warp_Amount -----})
  n135 = ChannelMaskParameter(None=TextureSample(UVs=Add(A=@n129, B=@n134), Tex=@n95).RGBA) {group=Wrap_LanternFest, maskChannel=Green, parameterName=LF_Dots_WarpMaskChannel, defaultValue=<rgba 0 1 0 0>}
  n136 = Time()
  n137 = ComponentMask(None=VectorParameter() {group=Wrap_LanternFest, parameterName=LF_Dots_WarpGradientspeed, defaultValue=<rgba -0.035 -0.05 0 1>}.RGB) {g=true, r=true}
  n138 = Multiply(
   A=Add(A=@n126.Radial Coordinates, B=Multiply(A=@n136, B=@n137))
   B=ScalarParameter() {defaultValue=1, group=Wrap_LanternFest, parameterName=LF_Dots_MaskTileRate}
  )
  n139 = ChannelMaskParameter(None=TextureSample(UVs=Add(A=@n138, B=@n134), Tex=@n95).RGBA) {group=Wrap_LanternFest, parameterName=LF_Dots_MaskChannel}
  n140 = Multiply(A=@n135.R, B=@n139.R)
  n141 = Saturate(None=Multiply(A=@n140, B=Add(A=@n140, B=@n139.R)))
  n142 = Multiply(
   A=SphereMask(
    A=@n115.UVs
    B=Constant() {r=0.5}
    Radius=ScalarParameter() {defaultValue=1, group=Wrap_LanternFest_UVs, parameterName=LF_Dots_SphereMaskRadius}
    Hardness=Constant()
   ) {hardnessPercent=0}
   B=Multiply(A=Saturate(None=@n120.Result), B=Multiply(A=@n121, B=@n141))
  )
  n143 = LinearInterpolate(
   A=LinearInterpolate(A=LinearInterpolate(A=@n85, B=@n5.RGB, Alpha=@n96.A), B=@n102, Alpha=LinearInterpolate(A=@n116, B=@n5.RGB, Alpha=@n96.A) {constB=0})
   B=@n102
   Alpha=LinearInterpolate(A=Round(None=@n142), B=@n5.RGB, Alpha=@n96.A)
  )
  n144 = VectorParameter() {group=Wrap_LanternFest, parameterName=LF_SphereInterior_Color, sortPriority=0, defaultValue=<rgba 0.7887 0.4242 0.0183 4>}
  n145 = LinearInterpolate(
   A=ScalarParameter() {defaultValue=0.05, group=Wrap_LanternFest, parameterName=LF_Sphere_Radius_Interior_Min}
   B=ScalarParameter() {defaultValue=0.1, group=Wrap_LanternFest, parameterName=LF_Sphere_Radius_Interior_Max}
   Alpha=@n105
  )
  n146 = LinearInterpolate(A=SphereMask(A=@n115.UVs, B=@n103, Radius=Saturate(None=@n145), Hardness=Constant()) {hardnessPercent=0}, B=@n5.RGB, Alpha=@n96.A)
  n147 = ScalarParameter() {defaultValue=0.5, group=Wrap_LanternFest, parameterName=TOD_MorningEvening, sortPriority=100}
  n148 = ScalarParameter() {defaultValue=1, group=Wrap_LanternFest, parameterName=TOD_Day, sortPriority=100}
  n149 = ScalarParameter() {defaultValue=0.25, group=Wrap_LanternFest, parameterName=TOD_Night, sortPriority=100}
  n150 = SetMaterialAttributes!(
   MaterialAttributes=@n2.MaterialAttributes
   BaseColor=@n3.RGB
   Roughness=ScalarParameter() {defaultValue=0.35, group=Function_Parallax, parameterName=FParallax_Rougness, sortPriority=0}
   Specular=ScalarParameter() {defaultValue=0.25, group=Function_Parallax, parameterName=FParallax_Specular, sortPriority=0}
   Metallic=ScalarParameter() {group=Function_Parallax, parameterName=FParallax_Metallic, sortPriority=0}
   EmissiveColor=Multiply(
    A=LinearInterpolate(A=@n143, B=@n144.RGB, Alpha=@n146)
    B=MaterialFunctionCall(Value_Morning=@n147, Value_Day=@n148, Value_Evening=@n147, Value_Night=@n149) {materialFunction=&Scalar_Time_of_Day}.Result
   )
  ) {attributeSetTypes [
    #2
    #3
    #4
    #5
    #6
   ]}
  n151 = FunctionInput() {bUsePreviewValueAsDefault=true, inputName="Customization Mask (_CM B)", inputType=FunctionInput_Scalar, sortPriority=24}
  FunctionOutput(None=StaticSwitchParameter(True=BlendMaterialAttributes(A=@n2.MaterialAttributes, B=@n150, Alpha=@n151) {vertexAttributeBlendType=UseA}, False=@n1) {group=Customization_Wrap, parameterName=Wrap_LanternFest}) {outputName=Result_1}
  Multiply(A=@n12)
  n152 = ScalarParameter() {defaultValue=-1, group=Function_Parallax, parameterName=FParallax_Depth_Amount_Layer_1_Weapon}
  n153 = ScalarParameter() {group=Function_Parallax, parameterName=FParallax_Depth_Amount_Layer_1_Vehicle}
  n154 = MaterialFunctionCall(Color=@n13) {materialFunction=&SplitComponents}
  Multiply(
   A=@n12
   B=LinearInterpolate(
    A=LinearInterpolate(A=@n152, B=@n153, Alpha=@n154.R)
    B=ScalarParameter() {group=Function_Parallax, parameterName=FParallax_Depth_Amount_Layer_1_Mech}
    Alpha=@n154.B
   )
  )
  Multiply()
  NamedRerouteDeclaration(None=@n90.A) {name=IsCharacter, nodeColor=<rgba 0.9237 1 0 1>}
  Reroute()
 ]}
