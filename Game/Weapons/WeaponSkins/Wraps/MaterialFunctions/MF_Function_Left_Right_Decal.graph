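# Wrap material function that layers a configurable decal onto weapons, vehicles, characters, backpacks and pickaxes, outputting blended material attributes and a decal mask.
# UVs come from mirrored left/right or single-side projection, with per-object-type transforms, safe zones and animated motion.
# Customization, curvature, facing and character masks limit the decal's BaseColor, roughness, metallic, normal and emissive blend.
# Optional emissive uses gradients and curve-atlas colors, with timestamp-driven reactive fades.
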MaterialFunction /Game/Weapons/WeaponSkins/Wraps/MaterialFunctions/MF_Function_Left_Right_Decal
graph {nodes [
  n1 = MaterialFunctionCall(Float2=NamedRerouteUsage()) {materialFunction=&BreakOutFloat2Components}
  n2 = ScalarParameter() {group=Function_Left_Right_Decal_Uvs, parameterName="USE_Pickaxe_UVS_Sets   0=off  1=on  ---------------", sliderMax=1, sortPriority=149}
  n3 = Min(A=@n1.G, B=@n2)
  Ceil(None=@n3)
  n4 = ScalarParameter() {group=Function_Left_Right_Decal_Uvs, parameterName="USE_BackPack_UVS_Sets 0=off  1=on  ---------------", sliderMax=1, sortPriority=175}
  n5 = Min(A=@n1.R, B=@n4)
  Ceil(None=@n5)
  Constant3Vector() {constant=<rgba 0 1 0 0>}
  Constant() {r=1}
  Constant()
  Constant()
  n6 = NamedRerouteUsage()
  n7 = NamedRerouteUsage()
  n8 = MaterialFunctionCall(Color=NamedRerouteUsage()) {materialFunction=&SplitComponents}
  n9 = ComponentMask(None=NamedRerouteUsage()) {b=true, g=true, r=true}
  n10 = NamedRerouteUsage()
  n11 = StaticSwitchParameter(
   True=ComponentMask(None=NamedRerouteUsage()) {b=true, g=true, r=true}
   False=StaticSwitchParameter(True=ComponentMask(None=@n10) {b=true, g=true, r=true}, False=NamedRerouteUsage()) {group=Function_Left_Right_Deca msk, parameterName=FDecal_UV_SafeZone_Front}
  ) {group=Function_Left_Right_Deca msk, parameterName=FDecal_UV_SafeZone_Middle}
  n12 = ScalarParameter() {group=Function_Left_Right_Decal_Uvs, parameterName="FDecal_UV_Animate_Sine  1=on", sliderMax=1, sortPriority=0}
  n13 = LinearInterpolate(B=NamedRerouteUsage(), Alpha=Saturate(None=Ceil(None=@n12)))
  n14 = NamedRerouteUsage()
  n15 = NamedRerouteUsage()
  n16 = ComponentMask(None=NamedRerouteUsage()) {a=true}
  n17 = Multiply(A=ComponentMask(None=NamedRerouteUsage()) {a=true}, B=@n15)
  n18 = NamedRerouteUsage()
  n19 = StaticSwitchParameter(
   True=Multiply(A=ComponentMask(None=@n14) {a=true}, B=@n15)
   False=StaticSwitchParameter(
    True=Multiply(A=@n16, B=@n15)
    False=StaticSwitchParameter(True=@n17, False=@n18) {group=Function_Left_Right_Deca msk, parameterName=FDecal_UV_SafeZone_Front}
   ) {group=Function_Left_Right_Deca msk, parameterName=FDecal_UV_SafeZone_Middle}
  ) {group=Function_Left_Right_Deca msk, parameterName=FDecal_UV_SafeZone_Back}
  n20 = NamedRerouteUsage()
  n21 = NamedRerouteUsage()
  n22 = Multiply(
   A=Time()
   B=ScalarParameter() {group=Function_Left_Right_Decal_Uvs, parameterName="FDecal_UV_Rotation_Speed  --------------", sortPriority=0}
  )
  n23 = Sine(None=@n22)
  n24 = ScalarParameter() {defaultValue=0.1, group=Function_Left_Right_Decal_Uvs, parameterName=FDecal_UV_Rotation_Speed Bounce Strength, sortPriority=0}
  n25 = ScalarParameter() {group=Function_Left_Right_Decal_Uvs, parameterName="FDecal_UV_Rotation_Speed  1=Bounce", sliderMax=1, sortPriority=0}
  n26 = Add(A=NamedRerouteUsage(), B=LinearInterpolate(A=@n22, B=Multiply(A=@n23, B=@n24), Alpha=Ceil(None=@n25)))
  n27 = MaterialFunctionCall!(
   For_Safe_Zones=StaticBool() {value=true}
   Normal_Map=@n7
   Caculated_Bounds=@n20
   Pre_Skinned_Normals=@n21
   Is Vehicle=@n8.R
   Is VehicleSide=@n8.R
   Is Mech=@n8.R
   Vehicle_Rotation
   UV_Rotation Bypass=@n26
   Mech_Rotation
   Uvs_Offset=Add(A=StaticSwitchParameter(True=@n9, False=@n11) {group=Function_Left_Right_Deca msk, parameterName=FDecal_UV_SafeZone_Back}, B=@n13)
   Offset-after rotation
   Uvs_TileRate=@n19
  ) {materialFunction=&/MF_Projection_Side_Normal}
  n28 = NamedRerouteUsage()
  n29 = MaterialFunctionCall!(
   Normal_Map=@n7
   Caculated_Bounds=@n20
   Pre_Skinned_Normals=@n21
   Is Vehicle=@n8.R
   Is Vehicle side=@n8.R
   Is Mech=@n8.R
   Offset-after rotation
   UV_Rotation_Right=@n26
   UV_Rotation_Left=Multiply(A=@n26) {constB=-1}
   Uvs_TileRate_Right=@n19
   Uvs_TileRate_Left=@n19
   Uvs_Offset_Right=@n28
   Uvs_Offset_Left=Multiply(A=@n28, B=Constant3Vector() {constant=<rgba -1 1 1 0>})
  ) {materialFunction=&/MF_Projection_Side_Left_Right_Decal_Normal}
  n30 = NamedRerouteDeclaration(None=StaticSwitchParameter(True=@n27.UVs, False=@n29.UVs) {group=Function_Left_Right_Deca msk, parameterName="FDecal_Projection-LR--->Side"}) {name=UVs, nodeColor=<rgba 0.125 0 1 1>}
  n31 = NamedRerouteUsage()
  n32 = TextureProperty(Tex=@n31)
  n33 = ScalarParameter() {defaultValue=1, group=Function_Left_Right_Decal, parameterName=FDecal_Miplevel_Adjustment, sortPriority=0}
  n34 = TextureSample!(
   UVs=LinearInterpolate(A=Saturate(None=@n30), B=@n30, Alpha=NamedRerouteUsage())
   Tex=@n31
   Level=MaterialFunctionCall!(UVs=NamedRerouteUsage(), Texture Size=Multiply(A=@n32, B=@n33)) {materialFunction=/Engine/Functions/Engine_MaterialFunctions01/Texturing/ComputeMipLevel}.MipLevel
   Apply View MipBias
  ) {mipValueMode=TMVM_MipLevel}
  n35 = ChannelMaskParameter(None=@n34.RGBA) {group=Function_Left_Right_Decal, maskChannel=Alpha, parameterName=FDecal_BaseColor_Tex_Mask_Channel, defaultValue=<rgba 0 0 0 1>}
  n36 = NamedRerouteUsage()
  n37 = Saturate(None=Add(A=@n36, B=@n27.Safe_Zone_Mask))
  n38 = Saturate(None=Add(A=@n36, B=@n29.Clamp_Mask))
  n39 = Saturate(None=StaticSwitchParameter(True=@n37, False=@n38) {group=Function_Left_Right_Deca msk, parameterName="FDecal_Projection-LR--->Side"})
  n40 = Multiply(
   A=StaticSwitchParameter(True=NamedRerouteUsage(), False=@n35.R) {group=Function_Left_Right_Decal, parameterName="FDecal_Decal---->FlipBook"}
   B=@n39
  )
  FunctionOutput(None=Multiply(A=@n6, B=@n40)) {outputName=Decal_Mask}
  n41 = StaticSwitchParameter(
   True=NamedRerouteUsage()
   False=StaticSwitchParameter(
    True=VectorParameter() {group=Function_Left_Right_Decal, parameterName=FDecal_BaseColor_Simple, defaultValue=<rgba 0 0 0 1>}.RGB
    False=@n34.RGB
   ) {group=Function_Left_Right_Decal, parameterName=FDecal_BaseColor__Simple}
  ) {group=Function_Left_Right_Decal, parameterName="FDecal_Decal---->FlipBook"}
  n42 = ScalarParameter() {group=Function_Left_Right_Decal, parameterName="FDecal_BaseColor=0-->Emissive=1", sortPriority=-5}
  n43 = ScalarParameter() {group=Function_Left_Right_Decal, parameterName=FDecal_Rougness_High}
  n44 = ScalarParameter() {defaultValue=1, group=Function_Left_Right_Decal, parameterName=FDecal_Rougness_Low}
  n45 = NamedRerouteUsage()
  n46 = NamedRerouteUsage()
  n47 = MaterialFunctionCall(
   UVs=NamedRerouteUsage()
   Texture Size=Multiply(
    A=TextureProperty(Tex=@n46)
    B=ScalarParameter() {defaultValue=1, group=Function_Left_Right_Decal, parameterName=FDecal_Miplevel_Adjustment, sortPriority=0}
   )
  ) {materialFunction=&ComputeMipLevel}
  n48 = VectorParameter() {group=Function_Left_Right_Decal_Uvs, parameterName=FDecal_Metalic_Mask_TileRate _UV, defaultValue=<rgba 1 1 0 1>}
  n49 = ComponentMask(None=@n48.RGB) {g=true, r=true}
  n50 = Time()
  n51 = VectorParameter() {group=Function_Left_Right_Decal_Uvs, parameterName=FDecal_Metalic_Mask_Speed_UV, defaultValue=<rgba 0 0 0 1>}
  n52 = ComponentMask(None=@n51.RGB) {g=true, r=true}
  n53 = TextureSample(UVs=Add(A=Multiply(A=@n45, B=@n49), B=Multiply(A=@n50, B=@n52)), Tex=@n46, Level=@n47.MipLevel) {mipValueMode=TMVM_MipLevel}
  n54 = StaticSwitchParameter(True=NamedRerouteUsage(), False=@n53.RGBA) {group=Function_Left_Right_Decal, parameterName="FDecal_Decal---->FlipBook"}
  n55 = StaticSwitchParameter(True=TextureSample(UVs=@n45, Tex=NamedRerouteUsage(), Level=@n47.MipLevel) {mipValueMode=TMVM_MipLevel}.RGBA, False=@n54) {desc=Turn off if not in use, group=Function_Left_Right_Decal_Uvs, parameterName="FDecal_Metalic ------>UVS_Static"}
  n56 = ChannelMaskParameter(None=@n55) {group=Function_Left_Right_Decal, maskChannel=Alpha, parameterName=FDecal_Roughness_Mask_Channel, defaultValue=<rgba 0 0 0 1>}
  n57 = NamedRerouteUsage()
  n58 = GetMaterialAttributes(None=NamedRerouteUsage()) {attributeGetTypes [#1, #2, #3]}
  n59 = NamedRerouteUsage()
  n60 = Multiply(A=@n37, B=@n27.Raw_Normal)
  n61 = Multiply(A=@n38, B=@n29.Raw_Normal)
  n62 = StaticSwitchParameter(True=@n57, False=@n58.Normal) {desc="By passes merged wrap normals. ", group=Function_Left_Right_Decal, parameterName=FDecal_Normal__Use_Root_Material_Normal}
  n63 = Add(
   A=@n62
   B=StaticSwitchParameter(
    True=@n59
    False=StaticSwitchParameter(True=@n60, False=@n61) {group=Function_Left_Right_Deca msk, parameterName="FDecal_Projection-LR--->Side"}
   ) {group=Function_Left_Right_Decal, parameterName="FDecal_Decal---->FlipBook"}
  )
  n64 = ScalarParameter() {defaultValue=1, group=Function_Left_Right_Decal, parameterName=FDecal_Normal_Intensty}
  n65 = ScalarParameter() {defaultValue=1, group=Function_Left_Right_Decal, parameterName="FDecal_Blackout_Emissive 1=off 0=on", sliderMax=1, sortPriority=-3}
  n66 = Add(A=OneMinus(None=@n40), B=@n65)
  n67 = Add(A=LinearInterpolate(B=@n41, Alpha=@n42), B=Multiply(A=NamedRerouteUsage(), B=Saturate(None=@n66)))
  n68 = ChannelMaskParameter(None=@n55) {
   group=Function_Left_Right_Decal
   maskChannel=Alpha
   parameterName=FDecal_Metalic_Emissive_Mask_Channel
   sortPriority=100
   defaultValue=<rgba 0 0 0 1>
  }
  n69 = Multiply(A=@n34.RGB, B=@n68.R)
  n70 = ScalarParameter() {defaultValue=1, group=Function_Left_Right_Decal, parameterName=FDecal_Metalic_Emissive_Mask_Brightness, sortPriority=100}
  n71 = StaticSwitchParameter(True=Add(A=@n67, B=Multiply(A=@n69, B=@n70)), False=@n67) {group=Function_Left_Right_Decal, parameterName=FDecal_Metalic_Emissive_Mask__Use, sortPriority=100}
  n72 = SetMaterialAttributes!(
   MaterialAttributes=SetMaterialAttributes!(MaterialAttributes=@n58.MaterialAttributes, BaseColor) {attributeSetTypes [#3]}
   BaseColor=LinearInterpolate(A=@n41, Alpha=@n42) {constB=0}
   Roughness=StaticSwitchParameter(
    True=ScalarParameter() {defaultValue=0.5, group=Function_Left_Right_Decal, parameterName=FDecal_Rougness}
    False=LinearInterpolate(A=@n43, B=@n44, Alpha=@n56.R)
   ) {defaultValue=true, group=Function_Left_Right_Decal, parameterName=FDecal_Roughness__Simple}
   Specular=ScalarParameter() {defaultValue=0.5, group=Function_Left_Right_Decal, parameterName=FDecal__Specular, sortPriority=150}
   Metallic=NamedRerouteUsage()
   Normal=StaticSwitchParameter(
    True=StaticSwitchParameter(True=@n57, False=@n58.Normal) {desc="By passes merged wrap normals. ", group=Function_Left_Right_Decal, parameterName=FDecal_Normal__Use_Root_Material_Normal}
    False=LinearInterpolate(A=@n62, B=@n63, Alpha=@n64)
   ) {defaultValue=true, group=Function_Left_Right_Decal, parameterName=FDecal_Normal__Simple}
   EmissiveColor=@n71
   CustomizedUV7
  ) {attributeSetTypes [
    #3
    #4
    #5
    #6
    #1
    #2
    #7
   ]}
  n73 = NamedRerouteUsage()
  n74 = ComponentMask(None=NamedRerouteUsage()) {r=true}
  n75 = LinearInterpolate(B=Saturate(None=Add(A=@n73, B=@n74)), Alpha=ComponentMask(None=NamedRerouteUsage()) {a=true}) {constA=1}
  n76 = Add(
   A=@n75
   B=Ceil(None=ScalarParameter() {group=Function_Left_Right_Deca msk, parameterName="FDecal_Remove_Character_Front_Msk   0=On", sliderMax=1, sortPriority=-10})
  )
  n77 = Multiply(
   A=NamedRerouteDeclaration(None=Saturate(None=@n76)) {name=character_Masking_Fnt/bck, nodeColor=<rgba 0 1 0.6016 1>}
   B=Multiply(A=@n6, B=@n40)
  )
  n78 = LinearInterpolate(
   A=ScalarParameter() {defaultValue=1, group=Function_Left_Right_Decal_Reactive, parameterName=FDecal_Opacity_Time__Start}
   B=ScalarParameter() {group=Function_Left_Right_Decal_Reactive, parameterName=FDecal_Opacity_Time__End}
   Alpha=NamedRerouteUsage()
  )
  n79 = BlendMaterialAttributes(
   A=NamedRerouteUsage()
   B=@n72
   Alpha=StaticSwitchParameter(True=Multiply(A=@n77, B=@n78), False=@n77) {group=Function_Left_Right_Decal_Reactive, parameterName=FDecal_Opacity_Reactive---------------------, sortPriority=0}
  ) {vertexAttributeBlendType=UseA}
  n80 = GetMaterialAttributes(None=@n79) {attributeGetTypes [#2, #3]}
  n81 = NamedRerouteUsage()
  n82 = ScalarParameter() {group=Function_Left_Right_Decal_Emissive, parameterName="FDecal_Remove_Character_Front_Msk Emissive   1=On", sliderMax=1, sortPriority=-10}
  n83 = LinearInterpolate(A=NamedRerouteUsage(), B=NamedRerouteUsage(), Alpha=Saturate(None=Ceil(None=@n82)))
  n84 = Add(A=NamedRerouteUsage(), B=OneMinus(None=@n83))
  n85 = SetMaterialAttributes!(
   MaterialAttributes=@n80.MaterialAttributes
   EmissiveColor=Add(A=@n80.EmissiveColor, B=Multiply(A=@n81, B=@n83))
   BaseColor=Multiply(A=@n80.BaseColor, B=Saturate(None=@n84))
  ) {attributeSetTypes [#2, #3]}
  FunctionOutput(None=StaticSwitchParameter(
    True=StaticSwitchParameter(True=@n85, False=@n79) {group=Function_Left_Right_Decal_Emissive, parameterName=FDecal_Emissive, sortPriority=0}
    False=NamedRerouteUsage()
   ) {group=Customization, parameterName=Function_Left_Right_Decal}) {outputName=Result_1}
  Multiply(A=@n36, B=@n59)
  n86 = VectorParameter() {
   group=Function_Left_Right_Deca msk
   parameterName=FDecal_TriPlanner_Curvature
   sortPriority=-5
   channelNames {a="", b="", g=Pits, r=peaks}
   defaultValue=<rgba 2 2 0 1>
  }
  n87 = Add(A=Multiply(A=NamedRerouteUsage(), B=@n86.peaks), B=Multiply(A=NamedRerouteUsage(), B=@n86.Pits))
  Multiply(A=@n87) {constB=2}
  NamedRerouteDeclaration(None=@n58.EmissiveColor) {name=Emissive_Channel, nodeColor=<rgba 0.8984 1 0 1>}
  n88 = VectorParameter() {group=Function_Left_Right_Decal_Uvs, parameterName=FDecal_UV_Scale_Right_Weapon, sortPriority=100, defaultValue=<rgba 1 1 1 1>}
  n89 = VectorParameter() {group=Function_Left_Right_Decal_Uvs, parameterName=FDecal_UV_Scale_Right_Vehicle, sortPriority=100, defaultValue=<rgba 1 1 1 1>}
  n90 = Multiply(A=@n89.RGB, B=@n89.RGB)
  n91 = VectorParameter() {group=Function_Left_Right_Decal_Uvs, parameterName=FDecal_UV_Scale_Right_VehicleSide, sortPriority=100, defaultValue=<rgba 1 1 1 1>}
  n92 = Multiply(A=@n91.RGB, B=@n91.RGB)
  n93 = MaterialFunctionCall(Float4=NamedRerouteUsage()) {materialFunction=&BreakOutFloat4Components}
  n94 = VectorParameter() {group=Function_Left_Right_Decal_Uvs, parameterName=FDecal_UV_Scale_Right_Mech, sortPriority=100, defaultValue=<rgba 1 1 1 1>}
  n95 = VectorParameter() {group=Function_Left_Right_Decal_Uvs, parameterName=FDecal_UV_Scale_Right_Character, sortPriority=100, defaultValue=<rgba 1 1 1 1>}
  n96 = LinearInterpolate(
   A=LinearInterpolate(A=Multiply(A=@n88.RGB, B=@n88.RGB), B=LinearInterpolate(A=@n90, B=@n92, Alpha=@n93.G), Alpha=@n93.R)
   B=LinearInterpolate(A=Multiply(A=@n94.RGB, B=@n94.RGB), B=Multiply(A=@n95.RGB, B=@n95.RGB), Alpha=@n93.A)
   Alpha=@n93.B
  )
  n97 = VectorParameter() {group=Function_Left_Right_Decal_Uvs, parameterName=FDecal_UV_Scale_Backpack, sortPriority=180, defaultValue=<rgba 1 1 1 1>}
  n98 = Multiply(A=@n97.RGB, B=@n97.RGB)
  n99 = VectorParameter() {group=Function_Left_Right_Decal_Uvs, parameterName=FDecal_UV_Scale_Pickaxe, sortPriority=150, defaultValue=<rgba 1 1 1 1>}
  NamedRerouteDeclaration(None=LinearInterpolate(A=LinearInterpolate(A=@n96, B=@n98, Alpha=@n5), B=Multiply(A=@n99.RGB, B=@n99.RGB), Alpha=@n3)) {name=TileRate, nodeColor=<rgba 0.0781 0 1 1>}
  n100 = PreSkinnedLocalBounds()
  n101 = Subtract(A=VertexInterpolator(VS=PreSkinnedPosition()).PS, B=@n100.Min)
  n102 = Subtract(A=@n100.Max, B=@n100.Max)
  NamedRerouteDeclaration(None=FunctionInput(Preview=Divide(A=@n101, B=@n102)) {bUsePreviewValueAsDefault=true, inputName=Calcated_Bounds, sortPriority=11}) {name=Bounds, nodeColor=<rgba 0.7578 1 0 1>}
  n103 = PreSkinnedNormal()
  NamedRerouteDeclaration(None=FunctionInput(Preview=VertexInterpolator(VS=@n103).PS) {bUsePreviewValueAsDefault=true, inputName=Pre-Skinned Local Normal, sortPriority=10}) {name=pre-Normals, nodeColor=<rgba 0 0.4609 1 1>}
  n104 = TextureSample!(UVs, Tex, Apply View MipBias) {samplerType=SAMPLERTYPE_Masks, texture=/Game/Weapons/FORT_Rifles/Textures/T_Scar_CM}
  n105 = FunctionInput(Preview=@n104.R) {bUsePreviewValueAsDefault=true, inputName=Customization Mask Primary, inputType=FunctionInput_Scalar, sortPriority=24}
  n106 = Add(
   A=@n105
   B=FunctionInput(Preview=@n104.B) {bUsePreviewValueAsDefault=true, inputName=Customization Mask Secondary, inputType=FunctionInput_Scalar, sortPriority=24}
  )
  n107 = ScalarParameter() {defaultValue=1, group=Function_Left_Right_Deca msk, parameterName=FDecal_Remove_Customization Mask Vehicles, sliderMax=1, sortPriority=-10}
  n108 = NamedRerouteUsage()
  n109 = Min(
   A=ScalarParameter() {group=Function_Left_Right_Deca msk, parameterName=FDecal_Remove_Customization Mask, sliderMax=1, sortPriority=-10}
   B=LinearInterpolate(
    B=@n107
    Alpha=MaterialFunctionCall(Float4=@n108) {materialFunction=&BreakOutFloat4Components}.R
   ) {constA=1}
  )
  n110 = Saturate(None=Add(A=@n106, B=@n109))
  n111 = NamedRerouteUsage()
  n112 = Multiply(A=@n110, B=@n111)
  n113 = StaticSwitchParameter(True=@n112, False=Multiply(A=NamedRerouteUsage(), B=@n111)) {group=Function_Left_Right_Deca msk, parameterName=FDecal_UV_SafeZone_Front}
  NamedRerouteDeclaration(None=StaticSwitchParameter(
    True=@n112
    False=StaticSwitchParameter(True=@n112, False=@n113) {group=Function_Left_Right_Deca msk, parameterName=FDecal_UV_SafeZone_Middle}
   ) {group=Function_Left_Right_Deca msk, parameterName=FDecal_UV_SafeZone_Back}) {name=Masking_Cust/Tri, nodeColor=<rgba 0.2188 0 1 1>}
  n114 = ScalarParameter() {defaultValue=1, group=Function_Left_Right_Decal_Emissive, parameterName="FDecal_Emissive_CC_Mask   0=ON", sliderMax=1}
  n115 = ScalarParameter() {defaultValue=86, group=Function_Left_Right_Decal_Emissive, parameterName="FDecal_Emissive_Gradent_CC  Index"}
  n116 = ScalarParameter() {group=Function_Left_Right_Decal_Emissive, parameterName=FDecal_Emissive_Warp_Streagth}
  n117 = NamedRerouteUsage()
  n118 = NamedRerouteUsage()
  n119 = ScalarParameter() {group=Function_Left_Right_Decal_Emissive, parameterName=FDecal_Emissive_Gradent_Rotation}
  n120 = VectorParameter() {group=Function_Left_Right_Decal_Emissive, parameterName=FDecal_Emissive_Msk_Offset, sortPriority=5}
  n121 = VectorParameter() {group=Function_Left_Right_Decal_Emissive, parameterName=FDecal_Emissive_Gradent_Scale, defaultValue=<rgba 1 1 1 1>}
  n122 = Multiply(
   A=Add(
    A=MaterialFunctionCall(UVs=@n118, "Rotation Angle (0-1)"=@n119) {materialFunction=&CustomRotator}.Rotated Values
    B=ComponentMask(None=@n120.RGB) {g=true, r=true}
   ) {constB=0}
   B=ComponentMask(None=@n121.RGB) {g=true, r=true}
  )
  n123 = ScalarParameter() {defaultValue=1, group=Function_Left_Right_Decal_Emissive, parameterName=FDecal_Emissive_Gradent_TileRate}
  n124 = Multiply(A=@n122, B=@n123)
  n125 = ScalarParameter() {defaultValue=1, group=Function_Left_Right_Decal_Emissive_Reactive, parameterName=FDecal_Emissive_Gradent_Time_End, sortPriority=210}
  n126 = Time()
  n127 = ScalarParameter() {group=Function_Left_Right_Decal_Emissive, parameterName=FDecal_Emissive_Gradent_Speed}
  n128 = StaticBoolParameter() {group=Function_Left_Right_Decal_Emissive_Reactive, parameterName=FDecal_Emissive_Gradent_Reactive---------------------, sortPriority=200}
  n129 = StaticSwitch(
   True=LinearInterpolate(
    A=ScalarParameter() {group=Function_Left_Right_Decal_Emissive_Reactive, parameterName=FDecal_Emissive_Gradent_Time_Start, sortPriority=210}
    B=@n125
    Alpha=NamedRerouteUsage()
   )
   False=Add(A=ScalarParameter() {group=Function_Left_Right_Decal_Emissive, parameterName=FDecal_Emissive_Gradent_Offset}, B=Multiply(A=@n126, B=@n127))
   Value=@n128
  )
  n130 = Add(A=@n122, B=@n129)
  n131 = ComponentMask(None=LinearInterpolate(A=Add(A=@n124, B=@n129), B=Multiply(A=@n130, B=@n123), Alpha=Ceil(None=@n123))) {r=true}
  n132 = StaticSwitchParameter(True=Add(A=Multiply(A=@n116, B=@n117), B=@n131), False=@n131) {group=Function_Left_Right_Decal_Emissive, parameterName=FDecal_Emissive_Warp_Metalic Msk}
  n133 = LinearInterpolate(
   A=Frac(None=@n132)
   B=@n132
   Alpha=ScalarParameter() {group=Function_Left_Right_Decal_Emissive, parameterName="FDecal_Emissive_Gradent_CC_Clamp  1=on"}
  )
  n134 = CurveAtlasRowParameter(CurveTime=@n133) {
   atlas=/Game/Weapons/WeaponSkins/Wraps/Textures/CruveAtlas_Wraps_01
   curve=/Game/Weapons/WeaponSkins/Wraps/Textures/0_ColorCurves/CC_Simple_Gradent
   defaultValue=86
   group=Function_Left_Right_Decal_Emissive
   parameterName=FDecal_Emissive_Gradent_CC
  }
  n135 = Divide(
   A=Frac(None=@n131)
   B=ScalarParameter() {defaultValue=1, group=Function_Left_Right_Decal_Emissive, parameterName=FDecal_Emissive_Gradent_Thickness}
  ) {constB=0.2}
  n136 = StaticSwitchParameter(
   True=ChannelMaskParameter(None=MaterialFunctionCall!(Index=@n115, CCinput=@n134, CCinput_Alpha=@n134, UV=@n133) {materialFunction=&/MF_Global_CC_Tex}.3V) {group=Function_Left_Right_Decal_Emissive, parameterName=FDecal_Emissive_Gradent_CC_Channel}.R
   False=Power(
    Base=OneMinus(None=@n135)
    Exp=ScalarParameter() {defaultValue=3, group=Function_Left_Right_Decal_Emissive, parameterName=FDecal_Emissive_Gradent_Falloff}
   )
  ) {group=Function_Left_Right_Decal_Emissive, parameterName="FDecal_Emissive_Gradent ---> use CC For Gradent"}
  n137 = Multiply(A=NamedRerouteUsage(), B=@n136)
  n138 = Multiply(A=NamedRerouteUsage(), B=@n137)
  n139 = Saturate(None=Multiply(A=@n138, B=Add(A=@n138, B=@n137)))
  n140 = ScalarParameter() {defaultValue=1, group=Function_Left_Right_Decal_Emissive, parameterName=FDecal_Emissive_Height_Contrast}
  n141 = StaticSwitchParameter(
   True=MaterialFunctionCall(Transition Phase=@n137, Height Texture=NamedRerouteUsage(), Contrast=@n140) {materialFunction=&HeightLerp}.Alpha
   False=StaticSwitchParameter(
    True=MaterialFunctionCall(In=@n139, Contrast=@n140) {materialFunction=&CheapContrast}.Result
    False=@n136
   ) {group=Function_Left_Right_Decal_Emissive, parameterName=FDecal_Emissive_Height___X+X, sortPriority=5}
  ) {group=Function_Left_Right_Decal_Emissive, parameterName=FDecal_Emissive_Height___Lerp, sortPriority=5}
  n142 = ScalarParameter() {group=Function_Left_Right_Decal_Emissive, parameterName=FDecal_Emissive_CC-Animated}
  n143 = Time()
  n144 = Multiply(A=ScalarParameter() {defaultValue=1, group=Function_Left_Right_Decal_Emissive, parameterName=FDecal_Emissive_CC_TileRate}, B=@n141)
  n145 = ScalarParameter() {group=Function_Left_Right_Decal_Emissive, parameterName=FDecal_Emissive_CC_Offset}
  n146 = Frac(None=Add(A=Multiply(A=@n142, B=@n143), B=Add(A=@n144, B=@n145)))
  n147 = CurveAtlasRowParameter(CurveTime=@n146) {
   atlas=&CruveAtlas_Wraps_01
   curve=/Game/Weapons/WeaponSkins/Wraps/Textures/C_GradSimple
   group=Function_Left_Right_Decal_Emissive
   parameterName=FDecal_Emissive_CC
  }
  n148 = MaterialFunctionCall(
   Index=ScalarParameter() {group=Function_Left_Right_Decal_Emissive, parameterName="FDecal_Emissive_CC  Index"}
   CCinput=@n147
   CCinput_Alpha=@n147
   UV=@n146
  ) {materialFunction=&MF_Global_CC_Tex}
  n149 = ScalarParameter() {defaultValue=1, group=Function_Left_Right_Decal_Emissive, parameterName=FDecal_Emissive_CC_Brightness}
  n150 = Multiply(A=@n148.Alpha, B=@n149)
  n151 = ScalarParameter() {defaultValue=1, group=Function_Left_Right_Decal_Emissive, parameterName="FDecal_Emissive_CC_Brightness USE CCAlpha 1=On", sliderMax=1}
  n152 = Ceil(None=@n151)
  n153 = Multiply(A=Saturate(None=Add(A=@n141, B=@n114)), B=Multiply(A=@n148.3V, B=LinearInterpolate(A=@n149, B=@n150, Alpha=@n152)))
  n154 = VectorParameter() {group=Function_Left_Right_Decal_Emissive, parameterName="FDecal-Emissive-Color--Alpha=Multplyer", defaultValue=<rgba 1 1 1 1>}
  n155 = Multiply(A=Multiply(A=@n141, B=@n154.RGB), B=@n154.A)
  n156 = StaticSwitchParameter(
   True=Multiply(A=NamedRerouteUsage(), B=NamedRerouteUsage())
   False=Multiply(
    A=StaticSwitchParameter(True=@n153, False=@n155) {group=Function_Left_Right_Decal_Emissive, parameterName="FDecal-Emissive-Color--->CCurve"}
    B=NamedRerouteUsage()
   )
  ) {
   desc=uses the Mask texture as a diffues emissive
   group=Function_Left_Right_Decal_Emissive
   parameterName="FDecal-Emissive--BC-Msk--->Simple-->Metalic_Msk_Tex"
   sortPriority=500
  }
  n157 = Multiply(
   A=@n156
   B=LinearInterpolate(
    A=ScalarParameter() {defaultValue=1, group=Function_Left_Right_Decal_Emissive_Reactive, parameterName=FDecal_Emissive_Brightness_Time__Start}
    B=ScalarParameter() {group=Function_Left_Right_Decal_Emissive_Reactive, parameterName=FDecal_Emissive_Brightness_Time__End}
    Alpha=NamedRerouteUsage()
   )
  )
  NamedRerouteDeclaration(None=Multiply(
    A=StaticSwitchParameter(True=@n157, False=@n156) {group=Function_Left_Right_Decal_Emissive_Reactive, parameterName=FDecal-Emissive_Reactive_Brightness---------------------}
    B=NamedRerouteUsage()
   )) {name=Emissive, nodeColor=<rgba 0.2656 1 0 1>}
  n158 = Step(
   Y=ScalarParameter() {defaultValue=0.6, group=Function_Left_Right_Decal_Emissive, parameterName=FDecal_Emissive_BC_Alpha_Clamp}
   X=NamedRerouteUsage()
  )
  NamedRerouteDeclaration(None=@n158) {name=Alpha_Msk, nodeColor=<rgba 1 0 0.9375 1>}
  n159 = ScalarParameter() {defaultValue=1, group=Function_Left_Right_Deca msk, parameterName="FDecal_TriPlanner_Mask 0=On 1=Off", sliderMax=1, sortPriority=-5}
  n160 = MaterialFunctionCall(Color=Abs(None=NamedRerouteUsage())) {materialFunction=&SplitComponents}
  n161 = MaterialFunctionCall(Color=NamedRerouteUsage()) {materialFunction=&SplitComponents}
  n162 = LinearInterpolate(A=@n160.R, B=LinearInterpolate(A=@n160.B, B=@n160.B, Alpha=@n161.G), Alpha=@n161.R)
  n163 = MaterialFunctionCall(
   In=Add(A=@n87, B=LinearInterpolate(A=@n162, B=@n160.G, Alpha=@n161.B))
   Contrast=ScalarParameter() {defaultValue=1, group=Function_Left_Right_Deca msk, parameterName=FDecal_TriPlanner_Mask_Contrast, sortPriority=-5}
  ) {materialFunction=&CheapContrast}
  NamedRerouteDeclaration(None=Saturate(None=Add(A=@n159, B=@n163.Result))) {name=Tri_planner_Msk, nodeColor=<rgba 0.3594 0 1 1>}
  n164 = FunctionInput(Preview=@n104.G) {bUsePreviewValueAsDefault=true, inputName=Mask Crevise, inputType=FunctionInput_Scalar, sortPriority=24, previewValue=<1 0 0 1>}
  n165 = Add(A=@n164) {constB=-0.5}
  NamedRerouteDeclaration(None=Saturate(None=Multiply(A=@n165) {constB=-1})) {name=Curve_Pits, nodeColor=<rgba 1 0 0.8906 1>}
  NamedRerouteDeclaration(None=Saturate(None=@n165)) {name=Curve_Peaks, nodeColor=<rgba 1 0 0.8906 1>}
  n166 = StaticSwitchParameter(
   True=LinearInterpolate(
    A=ScalarParameter() {group=Function_Left_Right_Decal_Emissive_Reactive, parameterName=FDecal_Emissive_Brightness_DarkenBase_Time_Start}
    B=ScalarParameter() {group=Function_Left_Right_Decal_Emissive_Reactive, parameterName=FDecal_Emissive_Brightness_DarkenBase_Time_End}
    Alpha=NamedRerouteUsage()
   )
   False=ScalarParameter() {group=Function_Left_Right_Decal_Emissive, parameterName=FDecal_Emissive_DarkenBase-------, sortPriority=4}
  ) {group=Function_Left_Right_Decal_Emissive_Reactive, parameterName=FDecal-Emissive_Reactive_Brightness---------------------}
  n167 = NamedRerouteUsage()
  n168 = Multiply(A=Multiply(A=@n166, B=@n141), B=@n167)
  n169 = Add(A=@n168, B=NamedRerouteUsage())
  n170 = ScalarParameter() {group=Function_Left_Right_Decal_Emissive, parameterName="FDecal_Emissive_DarkenBase-CCAlpha 1=Use", sliderMax=1, sortPriority=4}
  n171 = Ceil(None=@n170)
  n172 = StaticSwitchParameter(
   True=Multiply(A=@n166, B=@n167)
   False=StaticSwitchParameter(True=LinearInterpolate(A=@n168, B=@n169, Alpha=@n171), False=@n168) {group=Function_Left_Right_Decal_Emissive, parameterName="FDecal-Emissive-Color--->CCurve"}
  ) {
   desc=uses the Mask texture as a diffues emissive
   group=Function_Left_Right_Decal_Emissive
   parameterName="FDecal-Emissive--BC-Msk--->Simple-->Metalic_Msk_Tex"
   sortPriority=500
  }
  n173 = Saturate(None=Multiply(A=@n172, B=NamedRerouteUsage()))
  n174 = NamedRerouteUsage()
  n175 = ScalarParameter() {defaultValue=1, group=Function_Left_Right_Decal_Emissive, parameterName="FDecal_Emissive_DrawBehind BC 0=On", sliderMax=1, sortPriority=4}
  n176 = Saturate(None=Add(A=OneMinus(None=@n174), B=@n175))
  NamedRerouteDeclaration(None=LinearInterpolate(B=OneMinus(None=@n173), Alpha=@n176) {constA=1}) {name=Emissive_Darken_Base, nodeColor=<rgba 0 0.1328 1 1>}
  n177 = TextureObjectParameter() {
   group=Function_Left_Right_Decal
   parameterName=FDecal_BaseColor_Tex
   samplerType=SAMPLERTYPE_Masks
   texture=&T_BlackMask
  }
  NamedRerouteDeclaration(None=@n177) {name=Tex_Sample_BC, nodeColor=<rgba 0.1016 1 0 1>}
  n178 = NamedRerouteUsage()
  n179 = ChannelMaskParameter(None=StaticSwitchParameter(True=Multiply(A=@n178, B=@n39), False=@n34.RGBA) {group=Function_Left_Right_Decal, parameterName="FDecal_Decal---->FlipBook"}) {group=Function_Left_Right_Decal_Emissive, maskChannel=Alpha, parameterName=FDecal_Emissive_BC_Mask_Channel, defaultValue=<rgba 0 0 0 1>}
  NamedRerouteDeclaration(None=@n179.R) {name=Emissive_BC_Msk, nodeColor=<rgba 0.2187 1 0 1>}
  n180 = ScalarParameter() {
   defaultValue=1
   group=Function_Left_Right_Deca msk
   parameterName="FDecal_Remove_Customization Primary+Accent  0=Off"
   sliderMax=1
   sortPriority=-10
  }
  n181 = Add(A=@n105, B=Multiply(A=@n180, B=@n106))
  NamedRerouteDeclaration(None=Saturate(None=Add(A=@n109, B=@n181))) {name=Cust_Mask, nodeColor=<rgba 0.3125 1 0 1>}
  n182 = Constant() {r=1}
  n183 = StaticSwitchParameter(
   True=@n182
   False=StaticSwitchParameter(True=@n182, False=@n158) {group=Function_Left_Right_Decal_Emissive, parameterName=FDecal_Emissive_Height___X+X, sortPriority=5}
  ) {group=Function_Left_Right_Decal_Emissive, parameterName=FDecal_Emissive_Height___Lerp, sortPriority=5}
  NamedRerouteDeclaration(None=Multiply(A=NamedRerouteUsage(), B=@n183)) {name=Emissive_Msk_Simple, nodeColor=<rgba 0 1 0.1094 1>}
  n184 = TextureObjectParameter() {
   group=Function_Left_Right_Decal
   parameterName=FDecal_Normal_Tex
   samplerType=SAMPLERTYPE_Normal
   texture=&FlatNormalBlue
  }
  NamedRerouteDeclaration(None=@n184) {name=Tex_Normal, nodeColor=<rgba 0 1 0.8828 1>}
  NamedRerouteDeclaration(None=ChannelMaskParameter(None=@n54) {group=Function_Left_Right_Decal_Emissive, parameterName=FDecal_Emissive_Warp_Metalic_Msk_Tex_Channel}.R) {name=Emissive_Msk_Warp, nodeColor=<rgba 0.5234 1 0 1>}
  n185 = ScalarParameter() {defaultValue=-9999999, group=Function_Left_Right_Decal_Emissive_Reactive, parameterName=FDecal_Emissive_Gradent_Time_BP, sortPriority=212}
  n186 = ScalarParameter() {group=Function_Left_Right_Decal_Emissive_Reactive, parameterName=FDecal_Emissive_Gradent_Delay, sortPriority=211}
  n187 = ScalarParameter() {defaultValue=0.25, group=Function_Left_Right_Decal_Emissive_Reactive, parameterName=FDecal_Emissive_Gradent_Duration, sortPriority=211}
  n188 = Divide(A=Subtract(A=Time(), B=Add(A=@n185, B=@n186)), B=@n187)
  n189 = ScalarParameter() {defaultValue=3, group=Function_Left_Right_Decal_Emissive_Reactive, parameterName=FDecal_Emissive_Gradent_Falloff, sortPriority=211}
  NamedRerouteDeclaration(None=Power(Base=Saturate(None=@n188), Exp=@n189)) {name=reactive_Gradent, nodeColor=<rgba 0 0.5078 1 1>}
  NamedRerouteDeclaration(None=ChannelMaskParameter(None=@n54) {group=Function_Left_Right_Decal_Emissive, parameterName=FDecal_Emissive_Height__Metalic_Msk_Tex_Channel, sortPriority=5}.R) {name=Emissive_Msk, nodeColor=<rgba 0.5234 1 0 1>}
  n190 = Constant() {r=0.5}
  n191 = MaterialFunctionCall(Float2=@n122) {materialFunction=&BreakOutFloat2Components}
  n192 = VectorParameter() {
   group=Function_Left_Right_Decal_Emissive
   parameterName=FDecal_Emissive_Msk_Scale
   sortPriority=5
   channelNames {a=Y Falloff, b=X Falloff, g=Y Size, r=X Size}
   defaultValue=<rgba 0.5 0.5 1 1>
  }
  n193 = SphereMask(A=@n190, B=@n191.R, Radius=@n192.Y Size, Hardness=@n192.Y Size) {attenuationRadius=0.5}
  n194 = SphereMask(A=@n191.G, B=@n190, Radius=@n192.X Size, Hardness=@n192.X Size) {attenuationRadius=0.5}
  NamedRerouteDeclaration(None=Power(Base=Multiply(A=@n193, B=@n194))) {name=Decal_Msking, nodeColor=<rgba 0 1 0.5078 1>}
  NamedRerouteDeclaration(None=@n53.RGB) {name=Metalic_Msk, nodeColor=<rgba 0.3828 1 0 1>}
  n195 = TextureSample() {samplerType=SAMPLERTYPE_Normal, texture=/Game/Weapons/FORT_Rifles/T_SCAR_N}
  NamedRerouteDeclaration(None=FunctionInput(Preview=SetMaterialAttributes!(MaterialAttributes, Normal=@n195.RGB) {attributeSetTypes [#1]}) {bUsePreviewValueAsDefault=true, inputName=Material, inputType=FunctionInput_MaterialAttributes, sortPriority=-20}) {name=Root_Material, nodeColor=<rgba 1 0.2344 0 1>}
  n196 = ScalarParameter() {group=Function_Left_Right_Decal_Uvs, parameterName=FDecal_UV_Rotation_Right_Vehicle, sortPriority=101}
  n197 = ScalarParameter() {group=Function_Left_Right_Decal_Uvs, parameterName=FDecal_UV_Rotation_Right_VehicleSide, sortPriority=101}
  n198 = MaterialFunctionCall(Float4=NamedRerouteUsage()) {materialFunction=&BreakOutFloat4Components}
  n199 = ScalarParameter() {defaultValue=0.25, group=Function_Left_Right_Decal_Uvs, parameterName=FDecal_UV_Rotation_Right_Character, sortPriority=101}
  n200 = LinearInterpolate(
   A=LinearInterpolate(
    A=ScalarParameter() {group=Function_Left_Right_Decal_Uvs, parameterName=FDecal_UV_Rotation_Right_Weapon, sortPriority=101}
    B=LinearInterpolate(A=@n196, B=@n197, Alpha=@n198.G)
    Alpha=@n198.R
   )
   B=LinearInterpolate(
    A=ScalarParameter() {defaultValue=0.25, group=Function_Left_Right_Decal_Uvs, parameterName=FDecal_UV_Rotation_Right_Mech, sortPriority=101}
    B=@n199
    Alpha=@n198.A
   )
   Alpha=@n198.B
  )
  n201 = ScalarParameter() {group=Function_Left_Right_Decal_Uvs, parameterName=FDecal_UV_Rotation_Right_Backpack, sortPriority=180}
  NamedRerouteDeclaration(None=LinearInterpolate(
    A=LinearInterpolate(A=@n200, B=@n201, Alpha=@n5)
    B=ScalarParameter() {group=Function_Left_Right_Decal_Uvs, parameterName=FDecal_UV_Rotation_Right_Pickaxe, sortPriority=150}
    Alpha=@n3
   )) {name=Rotation, nodeColor=<rgba 0 0.5547 1 1>}
  n202 = ScalarParameter() {defaultValue=-9999999, group=Function_Left_Right_Decal_Emissive_Reactive, parameterName=FDecal_Emissive_Brightness_Time_BP, sortPriority=50}
  n203 = ScalarParameter() {group=Function_Left_Right_Decal_Emissive_Reactive, parameterName=FDecal_Emissive_Brightness_Delay}
  n204 = Divide(
   A=Subtract(A=Time(), B=Add(A=@n202, B=@n203))
   B=ScalarParameter() {defaultValue=0.25, group=Function_Left_Right_Decal_Emissive_Reactive, parameterName=FDecal_Emissive_Brightness_Duration}
  )
  NamedRerouteDeclaration(None=Power(
    Base=Saturate(None=@n204)
    Exp=ScalarParameter() {defaultValue=0.25, group=Function_Left_Right_Decal_Emissive_Reactive, parameterName=FDecal_Emissive_Brightness_Falloff}
   )) {name=Reactive_Brightness, nodeColor=<rgba 0.8516 0 1 1>}
  NamedRerouteDeclaration(None=@n176) {name=Alpha_MSk_B, nodeColor=<rgba 1 0 0.8203 1>}
  NamedRerouteDeclaration(None=FunctionInput(Preview=Constant4Vector() {constant=<rgba 0 0.2 0 0.5>}.RGBA) {bUsePreviewValueAsDefault=true, inputName=Object-SafeZone-Back, inputType=FunctionInput_Vector4, sortPriority=100}) {name=SafeZone_Back, nodeColor=<rgba 0.1016 1 0 1>}
  NamedRerouteDeclaration(None=FunctionInput(Preview=Constant4Vector() {constant=<rgba 0 0 0 0.5>}.RGBA) {bUsePreviewValueAsDefault=true, inputName=Object-SafeZone-Middle, inputType=FunctionInput_Vector4, sortPriority=101}) {name=SafeZone_Middle, nodeColor=<rgba 0.1016 1 0 1>}
  NamedRerouteDeclaration(None=FunctionInput(Preview=Constant4Vector() {constant=<rgba -0.2 0 0 0.5>}.RGBA) {bUsePreviewValueAsDefault=true, inputName=Object-SafeZone-Front, inputType=FunctionInput_Vector4, sortPriority=102}) {name=SafeZone_Front, nodeColor=<rgba 0.1016 1 0 1>}
  n205 = NamedRerouteUsage()
  n206 = Divide(A=ComponentMask(None=Add(A=@n205) {constB=-0.5}) {g=true}) {constB=-1}
  NamedRerouteDeclaration(None=Saturate(None=Ceil(None=@n206))) {name=Character mask, nodeColor=<rgba 0.9453 1 0 1>}
  n207 = ScalarParameter() {group=Function_Left_Right_Deca msk, parameterName="FDecal_Clamp_Mask disable 1=Active", sliderMax=1, sortPriority=-4}
  NamedRerouteDeclaration(None=Saturate(None=Round(None=@n207))) {name=ClampMask State, nodeColor=<rgba 0 0.0625 1 1>}
  n208 = LinearInterpolate(
   A=ScalarParameter() {defaultValue=1, group=Function_Left_Right_Decal, parameterName=FDecal_Metallic_High}
   B=ScalarParameter() {group=Function_Left_Right_Decal, parameterName=FDecal_Metallic_Low}
   Alpha=ChannelMaskParameter(None=@n55) {group=Function_Left_Right_Decal, parameterName=FDecal_Metalic_Mask_Channel}.R
  )
  NamedRerouteDeclaration(None=StaticSwitchParameter(True=ScalarParameter() {group=Function_Left_Right_Decal, parameterName=FDecal_Metallic}, False=Saturate(None=@n208)) {defaultValue=true, group=Function_Left_Right_Decal, parameterName=FDecal_Metalic__Simple}) {name=Metalic, nodeColor=<rgba 0.9922 0 1 1>}
  NamedRerouteDeclaration(None=@n148.Alpha) {name=CC_Alpha, nodeColor=<rgba 0 1 0.25 1>}
  n209 = VectorParameter() {group=Function_Left_Right_Decal_Uvs, parameterName=FDecal_UV_Offset_Right_Weapon, sortPriority=100, defaultValue=<rgba 0 0 0 1>}
  n210 = VectorParameter() {group=Function_Left_Right_Decal_Uvs, parameterName=FDecal_UV_Offset_Right_Vehicle, sortPriority=100, defaultValue=<rgba 0 0 0 1>}
  n211 = VectorParameter() {group=Function_Left_Right_Decal_Uvs, parameterName=FDecal_UV_Offset_Right_VehicleSide, sortPriority=100, defaultValue=<rgba 0 0 0 1>}
  n212 = MaterialFunctionCall(Float4=NamedRerouteUsage()) {materialFunction=&BreakOutFloat4Components}
  n213 = VectorParameter() {group=Function_Left_Right_Decal_Uvs, parameterName=FDecal_UV_Offset_Right_Mech, sortPriority=100, defaultValue=<rgba 0 0 0 1>}
  n214 = VectorParameter() {group=Function_Left_Right_Decal_Uvs, parameterName=FDecal_UV_Offset_Right_Character, sortPriority=100, defaultValue=<rgba 0 0 0 1>}
  n215 = LinearInterpolate(
   A=LinearInterpolate(A=@n209.RGB, B=LinearInterpolate(A=@n210.RGB, B=@n211.RGB, Alpha=@n212.G), Alpha=@n212.R)
   B=LinearInterpolate(A=@n213.RGB, B=@n214.RGB, Alpha=@n212.A)
   Alpha=@n212.B
  )
  n216 = VectorParameter() {group=Function_Left_Right_Decal_Uvs, parameterName=FDecal_UV_Offset_Right_Backpack, sortPriority=180, defaultValue=<rgba 0 0 0 1>}
  n217 = VectorParameter() {group=Function_Left_Right_Decal_Uvs, parameterName=FDecal_UV_Offset_Right_Pickaxe, sortPriority=150, defaultValue=<rgba 0 0 0 1>}
  NamedRerouteDeclaration(None=LinearInterpolate(A=LinearInterpolate(A=@n215, B=@n216.RGB, Alpha=@n5), B=@n217.RGB, Alpha=@n3)) {name=Offset, nodeColor=<rgba 0 0.5547 1 1>}
  n218 = ScalarParameter() {group=Function_Left_Right_Decal_Emissive, parameterName="FDecal_Emissive__TriPlanner_Mask  1=ON", sliderMax=1, sortPriority=1}
  n219 = LinearInterpolate(A=@n111, B=NamedRerouteUsage(), Alpha=Ceil(None=@n218))
  n220 = Multiply(A=@n110, B=@n219)
  n221 = StaticSwitchParameter(True=@n220, False=Multiply(A=@n219, B=NamedRerouteUsage())) {group=Function_Left_Right_Deca msk, parameterName=FDecal_UV_SafeZone_Front}
  NamedRerouteDeclaration(None=StaticSwitchParameter(
    True=@n220
    False=StaticSwitchParameter(True=@n220, False=@n221) {group=Function_Left_Right_Deca msk, parameterName=FDecal_UV_SafeZone_Middle}
   ) {group=Function_Left_Right_Deca msk, parameterName=FDecal_UV_SafeZone_Back}) {name=Masking_Cust/Tri_Emissive, nodeColor=<rgba 0.2188 0 1 1>}
  NamedRerouteDeclaration(None=@n163.Result) {name=Tri_planner_Msk_1, nodeColor=<rgba 0.3594 0 1 1>}
  NamedRerouteDeclaration(None=@n75) {name=character_Masking_Fnt/bck_1, nodeColor=<rgba 0 1 0.6016 1>}
  NamedRerouteDeclaration(None=@n219) {name=trimasking, nodeColor=<rgba 0.2891 0 1 1>}
  n222 = VectorParameter() {group=Function_Left_Right_Decal_Uvs, parameterName=FDecal_UV_Sine__SpeedXY, defaultValue=<rgba 0.1 0.1 0 1>}
  n223 = Multiply(A=Time(), B=@n222.RGB)
  n224 = VectorParameter() {group=Function_Left_Right_Decal_Uvs, parameterName=FDecal_UV_Sine__OffsetXY, defaultValue=<rgba 0 0 0 1>}
  n225 = ComponentMask(None=Sine(None=Add(A=@n223, B=@n224.RGB))) {g=true, r=true}
  n226 = VectorParameter() {group=Function_Left_Right_Decal_Uvs, parameterName=FDecal_UV_Sine_RangeXY_Weapon, defaultValue=<rgba 0 0 0 1>}
  n227 = VectorParameter() {group=Function_Left_Right_Decal_Uvs, parameterName=FDecal_UV_Sine_RangeXY_Vehicle, defaultValue=<rgba 0 0 0 1>}
  n228 = VectorParameter() {group=Function_Left_Right_Decal_Uvs, parameterName=FDecal_UV_Sine_RangeXY_VehicleSide, defaultValue=<rgba 0 0 0 1>}
  n229 = MaterialFunctionCall(Float4=NamedRerouteUsage()) {materialFunction=&BreakOutFloat4Components}
  n230 = VectorParameter() {group=Function_Left_Right_Decal_Uvs, parameterName=FDecal_UV_Sine_RangeXY_Mech, defaultValue=<rgba 0 0 0 1>}
  n231 = VectorParameter() {group=Function_Left_Right_Decal_Uvs, parameterName=FDecal_UV_Sine_RangeXY_Character, defaultValue=<rgba 0 0 0 1>}
  n232 = LinearInterpolate(
   A=LinearInterpolate(A=@n226.RGB, B=LinearInterpolate(A=@n227.RGB, B=@n228.RGB, Alpha=@n229.G), Alpha=@n229.R)
   B=LinearInterpolate(A=@n230.RGB, B=@n231.RGB, Alpha=@n229.A)
   Alpha=@n229.B
  )
  n233 = VectorParameter() {group=Function_Left_Right_Decal_Uvs, parameterName=FDecal_UV_Sine_RangeXY_Backpack, defaultValue=<rgba 0 0 0 1>}
  n234 = MaterialFunctionCall(Float2=NamedRerouteUsage()) {materialFunction=&BreakOutFloat2Components}
  n235 = VectorParameter() {group=Function_Left_Right_Decal_Uvs, parameterName=FDecal_UV_Sine_RangeXY_Pickaxe, defaultValue=<rgba 0 0 0 1>}
  n236 = ComponentMask(None=LinearInterpolate(A=LinearInterpolate(A=@n232, B=@n233.RGB, Alpha=@n234.R), B=@n235.RGB, Alpha=@n234.G)) {g=true, r=true}
  NamedRerouteDeclaration(None=AppendVector(A=Multiply(A=@n225, B=@n236), B=Constant())) {name=UVs_Sine_Range, nodeColor=<rgba 0 0.625 1 1>}
  n237 = Time()
  n238 = ScalarParameter() {defaultValue=1, group=Function_Left_Right_Decal-FlipBook, parameterName=FDecal_FlipBook_Speed}
  n239 = Saturate(None=Fmod(
    A=Multiply(A=@n237, B=@n238)
    B=ScalarParameter() {defaultValue=1, group=Function_Left_Right_Decal-FlipBook, parameterName=FDecal_FlipBook_Speed_Delay}
   ))
  n240 = Time()
  n241 = Add(
   A=ScalarParameter() {defaultValue=-9999999, group=Function_Left_Right_Decal-FlipBook, parameterName=FDecal_FlipBook_Time_FromBP}
   B=ScalarParameter() {group=Function_Left_Right_Decal-FlipBook, parameterName=FDecal_FlipBook_Time_Delay}
  )
  n242 = Saturate(None=Divide(
    A=Subtract(A=@n240, B=@n241)
    B=ScalarParameter() {defaultValue=1, group=Function_Left_Right_Decal-FlipBook, parameterName=FDecal_FlipBook_Time_Duration}
   ))
  n243 = Add(A=@n239, B=Add(A=@n242) {constB=-1})
  n244 = Add(
   A=StaticSwitchParameter(
    True=Saturate(None=@n243)
    False=StaticSwitchParameter(True=@n242, False=@n239) {group=Function_Left_Right_Decal-FlipBook, parameterName=FDecal_FlipBook_Time__Reactive}
   ) {group=Function_Left_Right_Decal-FlipBook, parameterName=FDecal_FlipBook_Time__Reactive + Speed}
   B=ScalarParameter() {group=Function_Left_Right_Decal-FlipBook, parameterName="FDecal_FlipBook_Frame ##"}
  )
  n245 = ScalarParameter() {defaultValue=2, group=Function_Left_Right_Decal-FlipBook, parameterName=FDecal_FlipBook_Number_Rows}
  n246 = ScalarParameter() {defaultValue=2, group=Function_Left_Right_Decal-FlipBook, parameterName=FDecal_FlipBook_Number_Columns}
  n247 = NamedRerouteUsage()
  n248 = NamedRerouteUsage()
  n249 = TextureProperty(Tex=@n247)
  n250 = Multiply(A=@n245, B=@n246)
  n251 = MaterialFunctionCall(
   UVs=NamedRerouteUsage()
   Texture Size=Multiply(
    A=Divide(A=@n249, B=@n250)
    B=ScalarParameter() {defaultValue=1, group=Function_Left_Right_Decal, parameterName=FDecal_Miplevel_Flip_Adjustment, sortPriority=0}
   )
  ) {materialFunction=&ComputeMipLevel}
  n252 = MaterialFunctionCall(
   "Animation  Phase (0-1)"=@n244
   Number of Rows=@n245
   Number of Columns=@n246
   Texture=@n247
   UVs=@n248
   MipBias/Level=@n251.MipLevel
   "Clamp Anim (see tooltip)"=StaticBool() {value=true}
  ) {materialFunction=&FlipBook}
  NamedRerouteDeclaration(None=AppendVector(
    A=NamedRerouteDeclaration(None=@n252.Result) {name=Flipbook_RGB, nodeColor=<rgba 0 0.0156 1 1>}
    B=NamedRerouteDeclaration(None=@n252.Alpha) {name=Flipbook_Alpha, nodeColor=<rgba 0 0.0156 1 1>}
   )) {name=Flipbook v4, nodeColor=<rgba 0 0.0156 1 1>}
  n253 = ScalarParameter() {defaultValue=-9999999, group=Function_Left_Right_Decal_Reactive, parameterName=FDecal_Opacity_Time_BP_Z}
  n254 = ScalarParameter() {group=Function_Left_Right_Decal_Reactive, parameterName=FDecal_Opacity_Delay}
  n255 = Divide(
   A=Subtract(A=Time(), B=Add(A=@n253, B=@n254))
   B=ScalarParameter() {defaultValue=0.25, group=Function_Left_Right_Decal_Reactive, parameterName=FDecal_Opacity_Duration}
  )
  NamedRerouteDeclaration(None=Power(
    Base=Saturate(None=@n255)
    Exp=ScalarParameter() {defaultValue=0.25, group=Function_Left_Right_Decal_Reactive, parameterName=FDecal_Opacity_Falloff}
   )) {name=Reactive_Opacity, nodeColor=<rgba 0.8516 0 1 1>}
  NamedRerouteDeclaration(None=StaticSwitchParameter(True=@n27.UVs No Offsets/Tiling, False=@n29.UVs No Offsets/Tiling) {group=Function_Left_Right_Deca msk, parameterName="FDecal_Projection-LR--->Side"}) {name=UVs_Raw, nodeColor=<rgba 0.125 0 1 1>}
  NamedRerouteDeclaration(None=FunctionInput() {inputName=isStateB, inputType=FunctionInput_Vector2, sortPriority=5}) {name=IsState_B, nodeColor=<rgba 0.1497 1 0.1633 1>}
  n256 = MaterialFunctionCall(
   "Animation  Phase (0-1)"=@n244
   Number of Rows=@n245
   Number of Columns=@n246
   Texture=NamedRerouteUsage()
   UVs=@n248
   MipBias/Level=@n251.MipLevel
  ) {materialFunction=&FlipBook}
  NamedRerouteDeclaration(None=AppendVector(A=@n256.Result, B=@n256.Result)) {name=Flipbook Msk, nodeColor=<rgba 0 0.0156 1 1>}
  n257 = NamedRerouteUsage()
  n258 = StaticBool() {value=true}
  NamedRerouteDeclaration(None=MaterialFunctionCall!(
    Normal_Map
    Normals_Source=MaterialFunctionCall("Animation  Phase (0-1)"=@n244, Number of Rows=@n245, Number of Columns=@n246, Texture=@n257, UVs=@n248, MipBias/Level=@n251.MipLevel) {materialFunction=&FlipBook}.Result
    Normals_Source_Use=@n258
    UVs
    Rotation_Comp
    UV_Rotation=NamedRerouteUsage()
    Pre_Skinned_Normals=NamedRerouteUsage()
    is_State=NamedRerouteUsage()
    "Left\\Right"=@n258
    No_Rotation
   ) {materialFunction=&/MF_UVs_Side_Fix_Normals}.Raw_Normal) {name=Flipbook Normals, nodeColor=<rgba 0 0.0156 1 1>}
  NamedRerouteDeclaration(None=FunctionInput(Preview=Constant4Vector().RGBA) {inputName=isState, inputType=FunctionInput_Vector4, sortPriority=5}) {name=isState, nodeColor=<rgba 0.9219 0 1 1>}
  n259 = TextureObjectParameter() {group=Function_Left_Right_Decal, parameterName=FDecal_Metalic_Mask__Tex, texture=&T_BlacksRGB}
  NamedRerouteDeclaration(None=@n259) {name=Tex_Sample_MSk, nodeColor=<rgba 0.1016 1 0 1>}
  NamedRerouteDeclaration(None=GetMaterialAttributes(None=FunctionInput() {bUsePreviewValueAsDefault=true, inputName=Material_Root, inputType=FunctionInput_MaterialAttributes, sortPriority=-20}) {attributeGetTypes [#1, #2]}.Normal) {name=Root_Normal, nodeColor=<rgba 0 1 0.3203 1>}
  n260 = Constant3Vector() {constant=<rgba -1 1 1 1>}
  n261 = ComponentMask(None=NamedRerouteUsage()) {b=true, g=true, r=true}
  n262 = MaterialFunctionCall(Float4=NamedRerouteUsage()) {materialFunction=&BreakOutFloat4Components}
  n263 = LinearInterpolate(A=@n261, B=Multiply(A=@n260, B=@n261), Alpha=@n262.R)
  n264 = ComponentMask(None=NamedRerouteUsage()) {b=true, g=true, r=true}
  n265 = ComponentMask(None=NamedRerouteUsage()) {b=true, g=true, r=true}
  n266 = Multiply(A=@n260, B=@n265)
  n267 = StaticSwitchParameter(
   True=LinearInterpolate(A=@n264, B=Multiply(A=@n260, B=@n264), Alpha=@n262.R)
   False=StaticSwitchParameter(True=LinearInterpolate(A=@n265, B=@n266, Alpha=@n262.R), False=NamedRerouteUsage()) {group=Function_Left_Right_Deca msk, parameterName=FDecal_UV_SafeZone_Front}
  ) {group=Function_Left_Right_Deca msk, parameterName=FDecal_UV_SafeZone_Middle}
  NamedRerouteDeclaration(None=Add(A=StaticSwitchParameter(True=@n263, False=@n267) {group=Function_Left_Right_Deca msk, parameterName=FDecal_UV_SafeZone_Back}, B=@n13)) {name=Offset_Left_Right, nodeColor=<rgba 0 1 0.9766 1>}
  NamedRerouteUsage()
  NamedRerouteUsage()
  NamedRerouteUsage()
  Round(None=@n12)
  StaticSwitchParameter(True=Multiply(A=NamedRerouteUsage(), B=NamedRerouteUsage()), False=@n183) {
   desc=uses the Mask texture as a diffues emissive
   group=Function_Left_Right_Decal_Emissive
   parameterName="FDecal-Emissive--BC-Msk--->Simple-->Metalic_Msk_Tex"
   sortPriority=500
  }
  StaticSwitchParameter(True=@n67, False=@n71) {group=Function_Left_Right_Decal_Emissive, parameterName=FDecal_Emissive, sortPriority=0}
 ]}
