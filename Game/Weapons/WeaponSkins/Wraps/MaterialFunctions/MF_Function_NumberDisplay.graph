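# Wrap material function that projects a two-digit number decal onto the sides of weapons, vehicles or mechs.
# Parameter sets chosen by type flags drive a mirrored left/right projection, whose UV halves select tens and ones cells from a digit flipbook.
# Projection, axis-normal and customization masks blend the digits' color, emissive and surface values into incoming attributes behind a static switch.
# An arbitrary-length digit chain remains unconnected to the output.
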MaterialFunction /Game/Weapons/WeaponSkins/Wraps/MaterialFunctions/MF_Function_NumberDisplay
graph {nodes [
  FunctionInput(Preview=Constant3Vector() {constant=<rgba 0 1 0 0>}) {bUsePreviewValueAsDefault=true, inputName=Forward Axis}
  n1 = FunctionInput(Preview=SetMaterialAttributes!(MaterialAttributes, Normal=TextureSample() {samplerType=SAMPLERTYPE_Normal, texture=/Game/Weapons/FORT_Rifles/T_SCAR_N}.RGB) {attributeSetTypes [#1]}) {bUsePreviewValueAsDefault=true, inputName=Material, inputType=FunctionInput_MaterialAttributes, sortPriority=-20}
  n2 = Constant()
  n3 = MaterialFunctionCall(
   R=FunctionInput(Preview=Constant()) {inputName=isVehicle, inputType=FunctionInput_Scalar, sortPriority=5}
   G=FunctionInput(Preview=@n2) {bUsePreviewValueAsDefault=true, inputName=Side Project Vehicle, inputType=FunctionInput_Scalar}
   B=FunctionInput(Preview=@n2) {bUsePreviewValueAsDefault=true, inputName=isMech, inputType=FunctionInput_Scalar, sortPriority=4}
  ) {materialFunction=&AppendMany}
  n4 = MaterialFunctionCall(Color=@n3.RGB) {materialFunction=&SplitComponents}
  n5 = ScalarParameter() {group=Function_NumberDisplay, parameterName=FNumDisplay_UV_Rotation_Right_Vehicle, sortPriority=101}
  n6 = ScalarParameter() {group=Function_NumberDisplay, parameterName=FNumDisplay_UV_Rotation_Right_VehicleSide, sortPriority=101}
  n7 = MaterialFunctionCall(Color=@n3.RGB) {materialFunction=&SplitComponents}
  n8 = LinearInterpolate(
   A=LinearInterpolate(
    A=ScalarParameter() {group=Function_NumberDisplay, parameterName=FNumDisplay_UV_Rotation_Right_Weapon, sortPriority=101}
    B=LinearInterpolate(A=@n5, B=@n6, Alpha=@n7.G)
    Alpha=@n7.R
   )
   B=ScalarParameter() {defaultValue=0.25, group=Function_NumberDisplay, parameterName=FNumDisplay_UV_Rotation_Right_Mech, sortPriority=101}
   Alpha=@n7.B
  )
  n9 = VectorParameter() {group=Function_NumberDisplay, parameterName=FNumDisplay_UV_Offset_Right_Weapon, sortPriority=100, defaultValue=<rgba 0 0 0 1>}
  n10 = VectorParameter() {group=Function_NumberDisplay, parameterName=FNumDisplay_UV_Offset_Right_Vehicle, sortPriority=100, defaultValue=<rgba 0 0 0 1>}
  n11 = VectorParameter() {group=Function_NumberDisplay, parameterName=FNumDisplay_UV_Offset_Right_VehicleSide, sortPriority=100, defaultValue=<rgba 0 0 0 1>}
  n12 = MaterialFunctionCall(Color=@n3.RGB) {materialFunction=&SplitComponents}
  n13 = VectorParameter() {group=Function_NumberDisplay, parameterName=FNumDisplay_UV_Offset_Right_Mech, sortPriority=100, defaultValue=<rgba 0 0 0 1>}
  n14 = LinearInterpolate(A=LinearInterpolate(A=@n9.RGB, B=LinearInterpolate(A=@n10.RGB, B=@n11.RGB, Alpha=@n12.G), Alpha=@n12.R), B=@n13.RGB, Alpha=@n12.B)
  n15 = FunctionInput(Preview=VertexInterpolator(VS=PreSkinnedNormal()).PS) {bUsePreviewValueAsDefault=true, inputName=Pre-Skinned Local Normal, sortPriority=10}
  n16 = PreSkinnedLocalBounds()
  n17 = Subtract(A=VertexInterpolator(VS=PreSkinnedPosition()).PS, B=@n16.Min)
  n18 = Subtract(A=@n16.Max, B=@n16.Max)
  n19 = VectorParameter() {group=Function_NumberDisplay, parameterName=FNumDisplay_UV_Scale_Right_Weapon, sortPriority=100, defaultValue=<rgba 1 0.6 0 0.5>}
  n20 = VectorParameter() {group=Function_NumberDisplay, parameterName=FNumDisplay_UV_Scale_Right_Vehicle, sortPriority=100, defaultValue=<rgba 1 0.6 0 0.5>}
  n21 = Multiply(A=@n20.RGB, B=@n20.RGB)
  n22 = VectorParameter() {group=Function_NumberDisplay, parameterName=FNumDisplay_UV_Scale_Right_VehicleSide, sortPriority=100, defaultValue=<rgba 1 0.6 0 0.5>}
  n23 = Multiply(A=@n22.RGB, B=@n22.RGB)
  n24 = MaterialFunctionCall(Color=@n3.RGB) {materialFunction=&SplitComponents}
  n25 = VectorParameter() {group=Function_NumberDisplay, parameterName=FNumDisplay_UV_Scale_Right_Mech, sortPriority=100, defaultValue=<rgba 1 0.6 0 0.5>}
  n26 = LinearInterpolate(
   A=LinearInterpolate(A=Multiply(A=@n19.RGB, B=@n19.RGB), B=LinearInterpolate(A=@n21, B=@n23, Alpha=@n24.G), Alpha=@n24.R)
   B=Multiply(A=@n25.RGB, B=@n25.RGB)
   Alpha=@n24.B
  )
  n27 = MaterialFunctionCall!(
   Normal_Map
   Caculated_Bounds=FunctionInput(Preview=Divide(A=@n17, B=@n18)) {bUsePreviewValueAsDefault=true, inputName=Calcated_Bounds, sortPriority=11}
   Pre_Skinned_Normals=@n15
   Is Vehicle=@n4.R
   Is Vehicle side=@n4.R
   Is Mech=@n4.R
   Offset-after rotation
   UV_Rotation_Right=@n8
   UV_Rotation_Left=Multiply(A=@n8) {constB=-1}
   Uvs_TileRate_Right=@n26
   Uvs_TileRate_Left=@n26
   Uvs_Offset_Right=@n14
   Uvs_Offset_Left=Multiply(A=@n14, B=Constant3Vector() {constant=<rgba -1 1 1 0>})
  ) {materialFunction=&/MF_Projection_Side_Left_Right_Decal_Normal}
  n28 = Multiply(A=@n27.UVs, B=Constant2Vector() {g=1, r=2}) {constB=2}
  n29 = MaterialFunctionCall(Float2=@n28) {materialFunction=&BreakOutFloat2Components}
  n30 = Add(A=@n29.R) {constB=-1}
  n31 = Round(None=ComponentMask(None=@n27.UVs) {r=true})
  n32 = MaterialFunctionCall(Float2=LinearInterpolate(
    A=@n28
    B=MaterialFunctionCall(X=@n30, Y=@n29.G) {materialFunction=&MakeFloat2}.Result
    Alpha=@n31
   )) {materialFunction=&BreakOutFloat2Components}
  n33 = ScalarParameter() {defaultValue=10, desc=Number of Colums in flip book, group=Function_NumberDisplay, parameterName=FNumDisplay_Digits_No_Columns}
  n34 = Divide(A=Constant() {r=1}, B=@n33)
  n35 = Constant() {r=10}
  n36 = ScalarParameter() {group=Function_NumberDisplay, parameterName=FNumDisplay_Digits_NumerValue, sortPriority=3}
  n37 = Divide(A=@n36, B=@n35) {constB=10}
  n38 = LinearInterpolate(A=Subtract(A=@n37, B=Frac(None=@n37)), B=Fmod(A=@n36, B=@n35), Alpha=@n31)
  n39 = MaterialFunctionCall(X=Add(A=Multiply(A=@n32.R, B=@n34), B=Multiply(A=@n34, B=@n38)) {constB=0}, Y=@n32.G) {materialFunction=&MakeFloat2}
  n40 = VectorParameter() {group=Function_NumberDisplay, parameterName=FNumDisplay_Digits_Offset_Lft, defaultValue=<rgba -0.1 0 0 1>}
  n41 = ComponentMask(None=Multiply(A=@n34, B=@n40.RGB)) {g=true, r=true}
  n42 = VectorParameter() {group=Function_NumberDisplay, parameterName=FNumDisplay_Digits_Offset_RT, defaultValue=<rgba 0.1 0 0 0>}
  n43 = ComponentMask(None=Multiply(A=@n34, B=@n42.RGB)) {g=true, r=true}
  n44 = TextureObjectParameter() {
   group=Function_NumberDisplay
   parameterName=FNumDisplay_Digits__Tex
   samplerType=SAMPLERTYPE_Masks
   texture=/Game/Global/Textures/Default/Flipbooks/T_NumberFlipBook
  }
  n45 = TextureSample!(UVs=LinearInterpolate(A=Add(A=@n39.Result, B=@n41), B=Add(A=@n39.Result, B=@n43), Alpha=@n31), Tex=@n44, Level, Apply View MipBias) {mipValueMode=TMVM_MipLevel}
  n46 = VectorParameter() {group=Function_NumberDisplay, parameterName=FNumDisplay_Digits_BaseColor_Color, defaultValue=<rgba 0 0 0 1>}
  n47 = Multiply(A=ChannelMaskParameter(None=@n45.RGBA) {group=Function_NumberDisplay, parameterName=FNumDisplay_Digits_BaseColor_Channel}.R, B=@n46.RGB)
  n48 = VectorParameter() {group=Function_NumberDisplay, parameterName=FNumDisplay_Digits_Emissive_Color, defaultValue=<rgba 0 0 0 1>}
  n49 = GetMaterialAttributes(None=@n1) {attributeGetTypes [#2]}
  n50 = SetMaterialAttributes!(
   MaterialAttributes=@n49.MaterialAttributes
   BaseColor=@n47
   Roughness=ScalarParameter() {defaultValue=0.5, group=Function_NumberDisplay, parameterName=FNumDisplay_Rougness, sortPriority=0}
   Specular=ScalarParameter() {defaultValue=0.5, group=Function_NumberDisplay, parameterName=FNumDisplay_Specular, sortPriority=0}
   Metallic=ScalarParameter() {group=Function_NumberDisplay, parameterName=FNumDisplay_Metallic, sortPriority=0}
   EmissiveColor=Multiply(
    A=ChannelMaskParameter(None=@n45.RGBA) {group=Function_NumberDisplay, parameterName=FNumDisplay_Digits_Emissive__Color_Channel}.R
    B=@n48.RGB
   )
  ) {attributeSetTypes [
    #2
    #3
    #4
    #5
    #6
   ]}
  n51 = FunctionInput() {bUsePreviewValueAsDefault=true, inputName="Customization Mask (_CM B)", inputType=FunctionInput_Scalar, sortPriority=24}
  n52 = ScalarParameter() {group=Function_NumberDisplay, parameterName=FNumDisplay_Remove_Customization Mask, sortPriority=-10}
  n53 = MaterialFunctionCall(Color=Abs(None=@n15)) {materialFunction=&SplitComponents}
  n54 = MaterialFunctionCall(Color=@n3.RGB) {materialFunction=&SplitComponents}
  n55 = LinearInterpolate(A=@n53.R, B=LinearInterpolate(A=@n53.B, B=@n53.B, Alpha=@n54.G), Alpha=@n54.R)
  n56 = Add(
   A=ScalarParameter() {defaultValue=1, group=Function_NumberDisplay, parameterName="FNumDisplay_TriPlanner_Mask 0=On 1=Off", sortPriority=-5}
   B=MaterialFunctionCall(
    In=LinearInterpolate(A=@n55, B=@n53.G, Alpha=@n54.B)
    Contrast=ScalarParameter() {defaultValue=1, group=Function_NumberDisplay, parameterName=FNumDisplay_TriPlanner_Mask_Contrast, sortPriority=-5}
   ) {materialFunction=&CheapContrast}.Result
  )
  n57 = ChannelMaskParameter(None=@n45.RGBA) {group=Function_NumberDisplay, parameterName=FNumDisplay_Digits_Mask_Channel}
  n58 = Multiply(A=Saturate(None=Add(A=@n51, B=@n52)), B=Multiply(A=Saturate(None=@n56), B=Multiply(A=@n57.R, B=@n27.Clamp_Mask)))
  FunctionOutput(None=StaticSwitchParameter(True=BlendMaterialAttributes(A=@n1, B=@n50, Alpha=@n58) {vertexAttributeBlendType=UseA}, False=@n1) {group=Customization, parameterName=Function_NumberDisplay}) {outputName=Result_1}
  LinearInterpolate(
   A=@n49.BaseColor
   B=@n47
   Alpha=ScalarParameter() {defaultValue=1, group=Function_NumberDisplay, parameterName="FNumDisplay_Digits_BaseColor__Color 1=ON"}
  )
  n59 = Add(A=@n27.UVs, B=Constant2Vector() {r=-0.25})
  n60 = Multiply(A=@n59, B=Constant2Vector() {g=0.1, r=2}) {constA=2, constB=0.1}
  n61 = Multiply(
   A=OneMinus(None=ComponentMask(None=@n60) {r=true})
   B=ScalarParameter() {defaultValue=2, group=Function_NumberDisplay, parameterName=FNumDisplay_Digits_HowMany, sortPriority=2}
  ) {constB=10}
  n62 = Subtract(A=@n61) {constB=0}
  n63 = Abs(None=@n36)
  n64 = Add(A=Ceil(None=Clamp(None=@n61) {clampMode=CMODE_ClampMin, maxDefault=0}))
  n65 = Add(A=@n64) {constB=3}
  n66 = Constant() {r=10}
  n67 = ScalarParameter() {
   defaultValue=-1
   group=Function_NumberDisplay
   parameterName="FNumDisplay_Digits_If_0_is_Displayed  -1=On  0=Off"
   sliderMin=-1
   sortPriority=1
  }
  n68 = If(A=Divide(A=@n63, B=Power(Base=@n66, Exp=@n65)), B=@n67, "A > B"=Constant() {r=1}, "A == B"=@n67, "A < B"=@n67)
  n69 = Clamp(None=@n61) {clampMode=CMODE_ClampMax, maxDefault=0}
  n70 = Clamp(None=Subtract(A=Abs(None=@n69)) {constB=0, desc=shift sampling to make room for decimal place}) {clampMode=CMODE_ClampMin, maxDefault=0}
  n71 = Clamp(None=Ceil(None=@n70))
  Multiply(B=Multiply(A=Ceil(None=@n62), B=Add(A=@n68, B=@n71)) {desc=remove 0's on the int side})
  n72 = MaterialFunctionCall(Float2=@n59) {materialFunction=&BreakOutFloat2Components}
  Multiply(
   A=SphereMask(A=Constant() {r=0.5}, B=Multiply(A=@n72.R) {constB=2}) {attenuationRadius=0.5}
   B=SphereMask(A=@n72.G, B=Constant() {r=0.5}) {attenuationRadius=0.5}
  )
  TextureObjectParameter() {
   group=Function_NumberDisplay
   parameterName=FNumDisplay_Normal_Tex
   samplerType=SAMPLERTYPE_Normal
   texture=&FlatNormalBlue
  }
  n73 = ComponentMask(None=@n60) {g=true}
  n74 = Power(Base=@n66, Exp=@n64) {desc="10-100,000,000"}
  n75 = Multiply(A=Frac(None=Divide(A=@n63, B=@n74))) {constB=100}
  n76 = Add(A=@n73, B=Divide(A=Floor(None=@n75)) {constB=10})
  n77 = Add(A=@n63, B=Divide(A=@n63) {constB=1000000}) {constB=0.0001, desc="offset the floating point error precision error... "}
  n78 = Power(Base=@n66, Exp=Floor(None=@n70)) {desc="10-100,000,000"}
  n79 = Multiply(A=Frac(None=Multiply(A=@n77, B=@n78)))
  n80 = Add(A=@n73, B=Divide(A=Floor(None=@n79)) {constB=10})
  TextureSample(UVs=AppendVector(A=Frac(None=@n61), B=LinearInterpolate(A=@n76, B=@n80, Alpha=@n71)), Tex=@n44) {mipValueMode=TMVM_MipLevel}
 ]}
