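# Material function for wrap skins that adds a four-layer, animated parallax pattern to weapons, vehicles, mechs and characters.
# Each layer offsets pre-skinned, curvature-warped bounds positions along the local camera vector, then samples through skew or side projection.
# IsState lerps per-type depth, tiling and rotation.
# Alpha-composited layers get curvature edge boost and time-of-day scaling, then feed emissive only or a masked material blend, behind static switches.
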MaterialFunction /Game/Weapons/WeaponSkins/Wraps/MaterialFunctions/MF_Function_Parallax
graph {nodes [
  n1 = NamedRerouteUsage()
  n2 = PreSkinnedLocalBounds()
  n3 = Subtract(A=VertexInterpolator(VS=PreSkinnedPosition()).PS, B=@n2.Min)
  n4 = Subtract(A=@n2.Max, B=@n2.Max)
  n5 = Add(A=FunctionInput(Preview=TextureSample() {samplerType=SAMPLERTYPE_Masks, texture=/Game/Weapons/FORT_Rifles/Textures/T_Scar_CM}.G) {bUsePreviewValueAsDefault=true, inputName=Curavture, inputType=FunctionInput_Scalar, sortPriority=24}) {constB=-0.5}
  n6 = Add(
   A=FunctionInput(Preview=Divide(A=@n3, B=@n4)) {bUsePreviewValueAsDefault=true, inputName=Calcated_Bounds, sortPriority=11}
   B=Multiply(A=@n5, B=ScalarParameter() {group=Function_Parallax, parameterName=FParallax_Cuvatur_Warp_amount, sortPriority=0})
  )
  n7 = Transform(None=CameraVectorWS()) {transformSourceType=TRANSFORMSOURCE_World, transformType=TRANSFORM_Local}
  n8 = Saturate(None=MaterialFunctionCall() {materialFunction=/Engine/Functions/Engine_MaterialFunctions02/WorldPositionOffset/ObjectScale}.Scale Y)
  n9 = FunctionInput(Preview=Constant()) {bUsePreviewValueAsDefault=true, inputName=IsState, inputType=FunctionInput_Vector4}
  n10 = MaterialFunctionCall(Color=@n9) {materialFunction=&SplitComponents}
  n11 = LinearInterpolate(
   A=LinearInterpolate(
    A=ScalarParameter() {defaultValue=0.15, group=Function_Parallax, parameterName=FParallax_Depth_Amount_Layer_3_Weapon}
    B=ScalarParameter() {group=Function_Parallax, parameterName=FParallax_Depth_Amount_Layer_3_Vehicle}
    Alpha=@n10.R
   )
   B=ScalarParameter() {group=Function_Parallax, parameterName=FParallax_Depth_Amount_Layer_3_Mech}
   Alpha=@n10.B
  )
  n12 = Add(A=@n6, B=Multiply(A=@n7, B=Multiply(A=@n8, B=@n11)))
  Add(A=@n1, B=@n12)
  n13 = FunctionInput() {inputName=IsVehicle, inputType=FunctionInput_Scalar, sortPriority=4}
  n14 = FunctionInput() {inputName=IsMech, inputType=FunctionInput_Scalar, sortPriority=4}
  ComponentMask(None=Saturate(None=Add(A=@n13, B=@n14))) {r=true}
  n15 = Saturate(None=Add(
    A=FunctionInput() {bUsePreviewValueAsDefault=true, inputName="Customization Mask (_CM B)", inputType=FunctionInput_Scalar, sortPriority=24}
    B=ScalarParameter() {group=Function_Parallax, parameterName=FParallax_Remove_Customization Mask, sortPriority=-10}
   ))
  n16 = NamedRerouteUsage()
  n17 = VectorParameter() {group=Function_Parallax, parameterName=FParallax_Tri_Mask_Color, sortPriority=0, defaultValue=<rgba 0 0 0 1>}
  n18 = ScalarParameter() {defaultValue=-2, group=Function_Parallax, parameterName=FParallax_Depth_Amount_Layer_0_Weapon}
  n19 = ScalarParameter() {group=Function_Parallax, parameterName=FParallax_Depth_Amount_Layer_0_Vehicle}
  n20 = MaterialFunctionCall(Color=@n9) {materialFunction=&SplitComponents}
  n21 = Multiply(
   A=@n8
   B=LinearInterpolate(
    A=LinearInterpolate(A=@n18, B=@n19, Alpha=@n20.R)
    B=ScalarParameter() {group=Function_Parallax, parameterName=FParallax_Depth_Amount_Layer_0_Mech}
    Alpha=@n20.B
   )
  )
  n22 = Add(A=@n1, B=Add(A=@n6, B=Multiply(A=@n7, B=@n21)))
  n23 = LinearInterpolate(
   A=ScalarParameter() {defaultValue=0.5, group=Function_Parallax, parameterName=FParallax_Skew_Amount, sortPriority=0}
   B=ScalarParameter() {defaultValue=0.5, group=Function_Parallax, parameterName=FParallax_Skew_Amount Character, sortPriority=0}
   Alpha=NamedRerouteUsage()
  )
  n24 = MaterialFunctionCall(Color=@n9) {materialFunction=&SplitComponents}
  n25 = VectorParameter() {group=Function_Parallax_UVS, parameterName=FParallaxUV_TileRate_Layer_0_Weapon, defaultValue=<rgba 1 2 0 1>}
  n26 = VectorParameter() {group=Function_Parallax_UVS, parameterName=FParallaxUV_TileRate_Layer_0_Vehicle, defaultValue=<rgba 1 2 0 1>}
  n27 = VectorParameter() {group=Function_Parallax_UVS, parameterName=FParallaxUV_TileRate_Layer_0_VehicleSide, defaultValue=<rgba 1 2 0 1>}
  n28 = MaterialFunctionCall(Color=@n9) {materialFunction=&SplitComponents}
  n29 = LinearInterpolate(
   A=Multiply(A=@n25.RGB, B=@n25.RGB)
   B=LinearInterpolate(A=Multiply(A=@n26.RGB, B=@n26.RGB), B=Multiply(A=@n27.RGB, B=@n27.RGB), Alpha=@n28.G)
   Alpha=@n28.R
  )
  n30 = VectorParameter() {group=Function_Parallax_UVS, parameterName=FParallaxUV_TileRate_Layer_0_Mech, defaultValue=<rgba 1 2 0 1>}
  n31 = Multiply(A=@n30.RGB, B=@n30.RGB)
  n32 = VectorParameter() {group=Function_Parallax_UVS, parameterName=FParallaxUV_TileRate_Mulpy_Character, defaultValue=<rgba 1 1 1 1>}
  n33 = LinearInterpolate(B=@n32.RGB, Alpha=NamedRerouteUsage()) {constA=1}
  n34 = ScalarParameter() {defaultValue=1, group=Function_Parallax, parameterName=FParallax_Skew_Alternate_Axies_Contrast, sortPriority=0}
  n35 = FunctionInput(Preview=VertexInterpolator(VS=PreSkinnedNormal()).PS) {bUsePreviewValueAsDefault=true, inputName=Pre-Skinned Local Normal, sortPriority=10}
  n36 = MaterialFunctionCall!(
   Normal_Map
   Caculated_Bounds=@n22
   Pre_Skinned_Normals=@n35
   Skew Amount=@n23
   Is Vehicle=@n24.R
   Uvs_Offset
   Uvs_TileRate=Multiply(A=LinearInterpolate(A=@n29, B=@n31, Alpha=@n28.B), B=@n33)
   Alternate Axies contrast=@n34
  ) {materialFunction=&/MF_Projection_Skew_Normal}
  n37 = ScalarParameter() {group=Function_Parallax_UVS, parameterName=FParallax_UV_Rotation_Layer_0_Vehicle}
  n38 = ScalarParameter() {group=Function_Parallax_UVS, parameterName=FParallax_UV_Rotation_Layer_0_VehicleSide}
  n39 = MaterialFunctionCall(Color=@n9) {materialFunction=&SplitComponents}
  n40 = LinearInterpolate(
   A=LinearInterpolate(
    A=ScalarParameter() {group=Function_Parallax_UVS, parameterName=FParallax_UV_Rotation_Layer_0_Weapon}
    B=LinearInterpolate(A=@n37, B=@n38, Alpha=@n39.G)
    Alpha=@n39.R
   )
   B=ScalarParameter() {defaultValue=0.25, group=Function_Parallax_UVS, parameterName=FParallax_UV_Rotation_Layer_0_Mech}
   Alpha=@n39.B
  )
  n41 = Time()
  n42 = Add(
   A=Multiply(A=VectorParameter() {group=Function_Parallax_UVS, parameterName=FParallax_UV_Speed 0, defaultValue=<rgba 0 0 0 1>}.RGB, B=@n41)
   B=NamedRerouteUsage()
  )
  n43 = TextureSampleParameter2D(UVs=Add(
    A=MaterialFunctionCall(UVs=@n36.UVs, "Rotation Angle (0-1)"=@n40) {materialFunction=&CustomRotator}.Rotated Values
    B=ComponentMask(None=@n42) {g=true, r=true}
   )) {
   group=Function_Parallax
   parameterName=FParallax_Text__Layer_0
   samplerSource=SSM_Wrap_WorldGroupSettings
   samplerType=SAMPLERTYPE_Masks
   texture=&T_BlackMask
  }
  n44 = Saturate(None=Add(A=@n43.A, B=ScalarParameter() {defaultValue=1, group=Function_Parallax, parameterName="FParallax_Text__Layer_0_useAlpha 0=on 1=off"}))
  n45 = ScalarParameter() {defaultValue=-1, group=Function_Parallax, parameterName=FParallax_Depth_Amount_Layer_1_Weapon}
  n46 = ScalarParameter() {group=Function_Parallax, parameterName=FParallax_Depth_Amount_Layer_1_Vehicle}
  n47 = MaterialFunctionCall(Color=@n9) {materialFunction=&SplitComponents}
  n48 = Multiply(
   A=@n8
   B=LinearInterpolate(
    A=LinearInterpolate(A=@n45, B=@n46, Alpha=@n47.R)
    B=ScalarParameter() {group=Function_Parallax, parameterName=FParallax_Depth_Amount_Layer_1_Mech}
    Alpha=@n47.B
   )
  )
  n49 = Add(A=@n1, B=Add(A=@n6, B=Multiply(A=@n7, B=@n48)))
  n50 = VectorParameter() {group=Function_Parallax_UVS, parameterName=FParallaxUV_TileRate_Layer_1_Weapon, defaultValue=<rgba 1 2 0 1>}
  n51 = VectorParameter() {group=Function_Parallax_UVS, parameterName=FParallaxUV_TileRate_Layer_1_Vehicle, defaultValue=<rgba 1 2 0 1>}
  n52 = VectorParameter() {group=Function_Parallax_UVS, parameterName=FParallaxUV_TileRate_Layer_1_VehicleSide, defaultValue=<rgba 1 2 0 1>}
  n53 = MaterialFunctionCall(Color=@n9) {materialFunction=&SplitComponents}
  n54 = LinearInterpolate(
   A=Multiply(A=@n50.RGB, B=@n50.RGB)
   B=LinearInterpolate(A=Multiply(A=@n51.RGB, B=@n51.RGB), B=Multiply(A=@n52.RGB, B=@n52.RGB), Alpha=@n53.G)
   Alpha=@n53.R
  )
  n55 = VectorParameter() {group=Function_Parallax_UVS, parameterName=FParallaxUV_TileRate_Layer_1_Mech, defaultValue=<rgba 1 2 0 1>}
  n56 = Multiply(A=@n55.RGB, B=@n55.RGB)
  n57 = MaterialFunctionCall(
   Caculated_Bounds=@n49
   Pre_Skinned_Normals=@n35
   Skew Amount=@n23
   Is Vehicle=@n24.R
   Uvs_TileRate=Multiply(A=LinearInterpolate(A=@n54, B=@n56, Alpha=@n53.B), B=@n33)
   Alternate Axies contrast=@n34
  ) {materialFunction=&MF_Projection_Skew_Normal}
  n58 = ScalarParameter() {group=Function_Parallax_UVS, parameterName=FParallax_UV_Rotation_Layer_1_Vehicle}
  n59 = ScalarParameter() {group=Function_Parallax_UVS, parameterName=FParallax_UV_Rotation_Layer_1_VehicleSide}
  n60 = MaterialFunctionCall(Color=@n9) {materialFunction=&SplitComponents}
  n61 = LinearInterpolate(
   A=LinearInterpolate(
    A=ScalarParameter() {group=Function_Parallax_UVS, parameterName=FParallax_UV_Rotation_Layer_1_Weapon}
    B=LinearInterpolate(A=@n58, B=@n59, Alpha=@n60.G)
    Alpha=@n60.R
   )
   B=ScalarParameter() {defaultValue=0.25, group=Function_Parallax_UVS, parameterName=FParallax_UV_Rotation_Layer_1_Mech}
   Alpha=@n60.B
  )
  n62 = Add(
   A=Multiply(A=VectorParameter() {group=Function_Parallax_UVS, parameterName=FParallax_UV_Speed 1, defaultValue=<rgba 0 0 0 1>}.RGB, B=@n41)
   B=NamedRerouteUsage()
  )
  n63 = TextureSampleParameter2D(UVs=Add(
    A=MaterialFunctionCall(UVs=@n57.UVs, "Rotation Angle (0-1)"=@n61) {materialFunction=&CustomRotator}.Rotated Values
    B=ComponentMask(None=@n62) {g=true, r=true}
   )) {
   group=Function_Parallax
   parameterName=FParallax_Text__Layer_1
   samplerSource=SSM_Wrap_WorldGroupSettings
   samplerType=SAMPLERTYPE_Masks
   texture=&T_BlackMask
  }
  n64 = ScalarParameter() {defaultValue=1, group=Function_Parallax, parameterName=FParallax_Text__Layer_1_Opacity}
  n65 = ScalarParameter() {defaultValue=-0.5, group=Function_Parallax, parameterName=FParallax_Depth_Amount_Layer_2_Weapon}
  n66 = ScalarParameter() {group=Function_Parallax, parameterName=FParallax_Depth_Amount_Layer_2_Vehicle}
  n67 = MaterialFunctionCall(Color=@n9) {materialFunction=&SplitComponents}
  n68 = Multiply(
   A=@n8
   B=LinearInterpolate(
    A=LinearInterpolate(A=@n65, B=@n66, Alpha=@n67.R)
    B=ScalarParameter() {group=Function_Parallax, parameterName=FParallax_Depth_Amount_Layer_2_Mech}
    Alpha=@n67.B
   )
  )
  n69 = Add(A=@n1, B=Add(A=@n6, B=Multiply(A=@n7, B=@n68)))
  n70 = VectorParameter() {group=Function_Parallax_UVS, parameterName=FParallaxUV_TileRate_Layer_2_Weapon, defaultValue=<rgba 1 2 0 1>}
  n71 = VectorParameter() {group=Function_Parallax_UVS, parameterName=FParallaxUV_TileRate_Layer_2_Vehicle, defaultValue=<rgba 1 2 0 1>}
  n72 = VectorParameter() {group=Function_Parallax_UVS, parameterName=FParallaxUV_TileRate_Layer_2_VehicleSide, defaultValue=<rgba 1 2 0 1>}
  n73 = MaterialFunctionCall(Color=@n9) {materialFunction=&SplitComponents}
  n74 = LinearInterpolate(
   A=Multiply(A=@n70.RGB, B=@n70.RGB)
   B=LinearInterpolate(A=Multiply(A=@n71.RGB, B=@n71.RGB), B=Multiply(A=@n72.RGB, B=@n72.RGB), Alpha=@n73.G)
   Alpha=@n73.R
  )
  n75 = VectorParameter() {group=Function_Parallax_UVS, parameterName=FParallaxUV_TileRate_Layer_2_Mech, defaultValue=<rgba 1 2 0 1>}
  n76 = Multiply(A=@n75.RGB, B=@n75.RGB)
  n77 = MaterialFunctionCall(
   Caculated_Bounds=@n69
   Pre_Skinned_Normals=@n35
   Skew Amount=@n23
   Is Vehicle=@n24.R
   Uvs_TileRate=Multiply(A=LinearInterpolate(A=@n74, B=@n76, Alpha=@n73.B), B=@n33)
   Alternate Axies contrast=@n34
  ) {materialFunction=&MF_Projection_Skew_Normal}
  n78 = ScalarParameter() {group=Function_Parallax_UVS, parameterName=FParallax_UV_Rotation_Layer_2_Vehicle}
  n79 = ScalarParameter() {group=Function_Parallax_UVS, parameterName=FParallax_UV_Rotation_Layer_2_VehicleSide}
  n80 = MaterialFunctionCall(Color=@n9) {materialFunction=&SplitComponents}
  n81 = LinearInterpolate(
   A=LinearInterpolate(
    A=ScalarParameter() {group=Function_Parallax_UVS, parameterName=FParallax_UV_Rotation_Layer_2_Weapon}
    B=LinearInterpolate(A=@n78, B=@n79, Alpha=@n80.G)
    Alpha=@n80.R
   )
   B=ScalarParameter() {defaultValue=0.25, group=Function_Parallax_UVS, parameterName=FParallax_UV_Rotation_Layer_2_Mech}
   Alpha=@n80.B
  )
  n82 = Add(
   A=Multiply(A=VectorParameter() {group=Function_Parallax_UVS, parameterName=FParallax_UV_Speed 2, defaultValue=<rgba 0 0 0 1>}.RGB, B=@n41)
   B=NamedRerouteUsage()
  )
  n83 = TextureSampleParameter2D(UVs=Add(
    A=MaterialFunctionCall(UVs=@n77.UVs, "Rotation Angle (0-1)"=@n81) {materialFunction=&CustomRotator}.Rotated Values
    B=ComponentMask(None=@n82) {g=true, r=true}
   )) {
   group=Function_Parallax
   parameterName=FParallax_Text__Layer_2
   samplerSource=SSM_Wrap_WorldGroupSettings
   samplerType=SAMPLERTYPE_Masks
   texture=&T_BlackMask
  }
  n84 = LinearInterpolate(
   A=LinearInterpolate(A=Multiply(A=@n43.RGB, B=@n44), B=@n63.RGB, Alpha=Multiply(A=@n63.A, B=@n64))
   B=@n83.RGB
   Alpha=Multiply(A=@n83.A, B=ScalarParameter() {defaultValue=1, group=Function_Parallax, parameterName=FParallax_Text__Layer_2_Opacity})
  )
  n85 = VectorParameter() {group=Function_Parallax_UVS, parameterName=FParallaxUV_TileRate_Layer_3_Weapon, defaultValue=<rgba 1 2 0 1>}
  n86 = VectorParameter() {group=Function_Parallax_UVS, parameterName=FParallaxUV_TileRate_Layer_3_Vehicle, defaultValue=<rgba 1 2 0 1>}
  n87 = VectorParameter() {group=Function_Parallax_UVS, parameterName=FParallaxUV_TileRate_Layer_3_VehicleSide, defaultValue=<rgba 1 2 0 1>}
  n88 = MaterialFunctionCall(Color=@n9) {materialFunction=&SplitComponents}
  n89 = LinearInterpolate(
   A=Multiply(A=@n85.RGB, B=@n85.RGB)
   B=LinearInterpolate(A=Multiply(A=@n86.RGB, B=@n86.RGB), B=Multiply(A=@n87.RGB, B=@n87.RGB), Alpha=@n88.G)
   Alpha=@n88.R
  )
  n90 = VectorParameter() {group=Function_Parallax_UVS, parameterName=FParallaxUV_TileRate_Layer_3_Mech, defaultValue=<rgba 1 2 0 1>}
  n91 = Multiply(A=@n90.RGB, B=@n90.RGB)
  n92 = MaterialFunctionCall(
   Caculated_Bounds=@n12
   Pre_Skinned_Normals=@n35
   Skew Amount=@n23
   Is Vehicle=@n24.R
   Uvs_TileRate=Multiply(A=LinearInterpolate(A=@n89, B=@n91, Alpha=@n88.B), B=@n33)
   Alternate Axies contrast=@n34
  ) {materialFunction=&MF_Projection_Skew_Normal}
  n93 = ScalarParameter() {group=Function_Parallax_UVS, parameterName=FParallax_UV_Rotation_Layer_3_Vehicle}
  n94 = ScalarParameter() {group=Function_Parallax_UVS, parameterName=FParallax_UV_Rotation_Layer_3_VehicleSide}
  n95 = MaterialFunctionCall(Color=@n9) {materialFunction=&SplitComponents}
  n96 = LinearInterpolate(
   A=LinearInterpolate(
    A=ScalarParameter() {group=Function_Parallax_UVS, parameterName=FParallax_UV_Rotation_Layer_3_Weapon}
    B=LinearInterpolate(A=@n93, B=@n94, Alpha=@n95.G)
    Alpha=@n95.R
   )
   B=ScalarParameter() {defaultValue=0.25, group=Function_Parallax_UVS, parameterName=FParallax_UV_Rotation_Layer_3_Mech}
   Alpha=@n95.B
  )
  n97 = Add(
   A=Multiply(A=VectorParameter() {group=Function_Parallax_UVS, parameterName=FParallax_UV_Speed 3, defaultValue=<rgba 0 0 0 1>}.RGB, B=@n41)
   B=NamedRerouteUsage()
  )
  n98 = TextureSampleParameter2D(UVs=Add(
    A=MaterialFunctionCall(UVs=@n92.UVs, "Rotation Angle (0-1)"=@n96) {materialFunction=&CustomRotator}.Rotated Values
    B=ComponentMask(None=@n97) {g=true, r=true}
   )) {
   group=Function_Parallax
   parameterName=FParallax_Text__Layer_3
   samplerSource=SSM_Wrap_WorldGroupSettings
   samplerType=SAMPLERTYPE_Masks
   texture=&T_BlackMask
  }
  n99 = Multiply(A=@n98.A, B=ScalarParameter() {defaultValue=1, group=Function_Parallax, parameterName=FParallax_Text__Layer_3_Opacity})
  n100 = VectorParameter() {group=Function_Parallax, parameterName=FParallax_Tri_Mask_Color, sortPriority=0, defaultValue=<rgba 0 0 0 1>}
  n101 = MaterialFunctionCall(Color=@n9) {materialFunction=&SplitComponents}
  n102 = ScalarParameter() {defaultValue=1, group=Function_Parallax_UVS, parameterName=FParallax_UV_TileRate_Layer_3_Weapon}
  n103 = MaterialFunctionCall(Color=@n9) {materialFunction=&SplitComponents}
  n104 = LinearInterpolate(
   A=ScalarParameter() {defaultValue=1, group=Function_Parallax_UVS, parameterName=FParallax_UV_TileRate_Layer_3_Vehicle}
   B=ScalarParameter() {defaultValue=1, group=Function_Parallax_UVS, parameterName=FParallax_UV_TileRate_Layer_3_VehicleSide}
   Alpha=@n103.G
  )
  n105 = ScalarParameter() {defaultValue=1, group=Function_Parallax_UVS, parameterName=FParallax_UV_TileRate_Layer_3_Mech}
  n106 = ScalarParameter() {defaultValue=1, group=Function_Parallax_UVS, parameterName=FParallax_UV_TileRate_Layer_3_Character}
  n107 = NamedRerouteUsage()
  n108 = MaterialFunctionCall!(
   For_Safe_Zones
   Normal_Map
   Caculated_Bounds=@n12
   Pre_Skinned_Normals=@n35
   Is Vehicle=@n101.R
   Is VehicleSide=@n101.R
   Is Mech=@n101.R
   Vehicle_Rotation
   UV_Rotation Bypass=@n96
   Mech_Rotation
   Uvs_Offset=@n97
   Offset-after rotation
   Uvs_TileRate=LinearInterpolate(A=LinearInterpolate(A=@n102, B=@n104, Alpha=@n103.R), B=LinearInterpolate(A=@n105, B=@n106, Alpha=@n107), Alpha=@n103.B)
  ) {materialFunction=&/MF_Projection_Side_Normal}
  n109 = TextureObjectParameter() {
   group=Function_Parallax
   parameterName=FParallax_Text__Layers
   samplerType=SAMPLERTYPE_Masks
   texture=&T_BlackMask
  }
  n110 = TextureSample(UVs=@n108.UVs, Tex=@n109) {samplerType=SAMPLERTYPE_Masks}
  n111 = TextureSampleParameter2D(UVs=@n108.UVs) {
   group=Function_Parallax
   parameterName=FParallax_Text__Layer_3
   samplerSource=SSM_Wrap_WorldGroupSettings
   samplerType=SAMPLERTYPE_Masks
   texture=&T_BlackMask
  }
  n112 = StaticBoolParameter() {group=Function_Parallax, parameterName="FParallax_Text__Layer----->Share Texture"}
  n113 = StaticSwitch(True=@n110.RGBA, False=@n111.RGBA, Value=@n112)
  n114 = LinearInterpolate(A=@n100.RGB, B=ComponentMask(None=@n113) {b=true, g=true, r=true}, Alpha=ComponentMask(None=@n113) {a=true})
  n115 = ScalarParameter() {group=Function_Parallax, parameterName="FParallax_Tri_Mask__Color_Use_Texture_03  0=off  1=On", sortPriority=0}
  n116 = ScalarParameter() {defaultValue=1, group=Function_Parallax_UVS, parameterName=FParallax_UV_TileRate_Layer_0_Weapon}
  n117 = MaterialFunctionCall(Color=@n9) {materialFunction=&SplitComponents}
  n118 = LinearInterpolate(
   A=ScalarParameter() {defaultValue=1, group=Function_Parallax_UVS, parameterName=FParallax_UV_TileRate_Layer_0_Vehicle}
   B=ScalarParameter() {defaultValue=1, group=Function_Parallax_UVS, parameterName=FParallax_UV_TileRate_Layer_0_VehicleSide}
   Alpha=@n117.G
  )
  n119 = ScalarParameter() {defaultValue=1, group=Function_Parallax_UVS, parameterName=FParallax_UV_TileRate_Layer_0_Mech}
  n120 = ScalarParameter() {defaultValue=1, group=Function_Parallax_UVS, parameterName=FParallax_UV_TileRate_Layer_0_Character}
  n121 = NamedRerouteUsage()
  n122 = MaterialFunctionCall(
   Caculated_Bounds=@n22
   Pre_Skinned_Normals=@n35
   Is Vehicle=@n101.R
   Is VehicleSide=@n101.R
   Is Mech=@n101.R
   UV_Rotation Bypass=@n40
   Uvs_Offset=@n42
   Uvs_TileRate=LinearInterpolate(A=LinearInterpolate(A=@n116, B=@n118, Alpha=@n117.R), B=LinearInterpolate(A=@n119, B=@n120, Alpha=@n121), Alpha=@n117.B)
  ) {materialFunction=&MF_Projection_Side_Normal}
  n123 = TextureSample(UVs=@n122.UVs, Tex=@n109) {samplerType=SAMPLERTYPE_Masks}
  n124 = TextureSampleParameter2D(UVs=@n122.UVs) {
   group=Function_Parallax
   parameterName=FParallax_Text__Layer_0
   samplerSource=SSM_Wrap_WorldGroupSettings
   samplerType=SAMPLERTYPE_Masks
   texture=&T_BlackMask
  }
  n125 = StaticSwitch(True=@n123.RGB, False=@n124.RGB, Value=@n112)
  n126 = StaticSwitch(True=@n123.A, False=@n124.A, Value=@n112)
  n127 = ScalarParameter() {defaultValue=1, group=Function_Parallax, parameterName="FParallax_Text__Layer_0_useAlpha 0=on 1=off"}
  n128 = Multiply(
   A=Saturate(None=Add(A=@n126, B=@n127))
   B=ScalarParameter() {defaultValue=1, group=Function_Parallax, parameterName=FParallax_Text__Layer_0_Opacity}
  )
  n129 = ScalarParameter() {defaultValue=1, group=Function_Parallax_UVS, parameterName=FParallax_UV_TileRate_Layer_1_Weapon}
  n130 = MaterialFunctionCall(Color=@n9) {materialFunction=&SplitComponents}
  n131 = LinearInterpolate(
   A=ScalarParameter() {defaultValue=1, group=Function_Parallax_UVS, parameterName=FParallax_UV_TileRate_Layer_1_Vehicle}
   B=ScalarParameter() {defaultValue=1, group=Function_Parallax_UVS, parameterName=FParallax_UV_TileRate_Layer_1_VehicleSide}
   Alpha=@n130.G
  )
  n132 = ScalarParameter() {defaultValue=1, group=Function_Parallax_UVS, parameterName=FParallax_UV_TileRate_Layer_1_Mech}
  n133 = ScalarParameter() {defaultValue=1, group=Function_Parallax_UVS, parameterName=FParallax_UV_TileRate_Layer_1_Character}
  n134 = NamedRerouteUsage()
  n135 = MaterialFunctionCall(
   Caculated_Bounds=@n49
   Pre_Skinned_Normals=@n35
   Is Vehicle=@n101.R
   Is VehicleSide=@n101.R
   Is Mech=@n101.R
   UV_Rotation Bypass=@n61
   Uvs_Offset=@n62
   Uvs_TileRate=LinearInterpolate(A=LinearInterpolate(A=@n129, B=@n131, Alpha=@n130.R), B=LinearInterpolate(A=@n132, B=@n133, Alpha=@n134), Alpha=@n130.B)
  ) {materialFunction=&MF_Projection_Side_Normal}
  n136 = TextureSample(UVs=@n135.UVs, Tex=@n109) {samplerType=SAMPLERTYPE_Masks}
  n137 = TextureSampleParameter2D(UVs=@n135.UVs) {
   group=Function_Parallax
   parameterName=FParallax_Text__Layer_1
   samplerSource=SSM_Wrap_WorldGroupSettings
   samplerType=SAMPLERTYPE_Masks
   texture=&T_BlackMask
  }
  n138 = StaticSwitch(True=@n136.A, False=@n137.A, Value=@n112)
  n139 = ScalarParameter() {defaultValue=1, group=Function_Parallax, parameterName=FParallax_Text__Layer_1_Opacity}
  n140 = ScalarParameter() {defaultValue=1, group=Function_Parallax_UVS, parameterName=FParallax_UV_TileRate_Layer_2_Weapon}
  n141 = MaterialFunctionCall(Color=@n9) {materialFunction=&SplitComponents}
  n142 = LinearInterpolate(
   A=ScalarParameter() {defaultValue=1, group=Function_Parallax_UVS, parameterName=FParallax_UV_TileRate_Layer_2_Vehicle}
   B=ScalarParameter() {defaultValue=1, group=Function_Parallax_UVS, parameterName=FParallax_UV_TileRate_Layer_2_VehicleSide}
   Alpha=@n141.G
  )
  n143 = ScalarParameter() {defaultValue=1, group=Function_Parallax_UVS, parameterName=FParallax_UV_TileRate_Layer_2_Mech}
  n144 = ScalarParameter() {defaultValue=1, group=Function_Parallax_UVS, parameterName=FParallax_UV_TileRate_Layer_2_Character}
  n145 = NamedRerouteUsage()
  n146 = MaterialFunctionCall(
   Caculated_Bounds=@n69
   Pre_Skinned_Normals=@n35
   Is Vehicle=@n101.R
   Is VehicleSide=@n101.R
   Is Mech=@n101.R
   UV_Rotation Bypass=@n81
   Uvs_Offset=@n82
   Uvs_TileRate=LinearInterpolate(A=LinearInterpolate(A=@n140, B=@n142, Alpha=@n141.R), B=LinearInterpolate(A=@n143, B=@n144, Alpha=@n145), Alpha=@n141.B)
  ) {materialFunction=&MF_Projection_Side_Normal}
  n147 = TextureSample(UVs=@n146.UVs, Tex=@n109) {samplerType=SAMPLERTYPE_Masks}
  n148 = TextureSampleParameter2D(UVs=@n146.UVs) {
   group=Function_Parallax
   parameterName=FParallax_Text__Layer_2
   samplerSource=SSM_Wrap_WorldGroupSettings
   samplerType=SAMPLERTYPE_Masks
   texture=&T_BlackMask
  }
  n149 = LinearInterpolate(
   A=LinearInterpolate(A=Multiply(A=@n125, B=@n128), B=StaticSwitch(True=@n136.RGB, False=@n137.RGB, Value=@n112), Alpha=Multiply(A=@n138, B=@n139))
   B=StaticSwitch(True=@n147.RGB, False=@n148.RGB, Value=@n112)
   Alpha=Multiply(
    A=StaticSwitch(True=@n147.A, False=@n148.A, Value=@n112)
    B=ScalarParameter() {defaultValue=1, group=Function_Parallax, parameterName=FParallax_Text__Layer_2_Opacity}
   )
  )
  n150 = StaticSwitch(True=@n110.RGB, False=@n111.RGB, Value=@n112)
  n151 = Multiply(
   A=StaticSwitch(True=@n110.A, False=@n111.A, Value=@n112)
   B=ScalarParameter() {defaultValue=1, group=Function_Parallax, parameterName=FParallax_Text__Layer_3_Opacity}
  )
  n152 = MaterialFunctionCall(Color=Abs(None=@n35)) {materialFunction=&SplitComponents}
  n153 = MaterialFunctionCall(Color=@n9) {materialFunction=&SplitComponents}
  n154 = LinearInterpolate(A=LinearInterpolate(A=@n152.R, B=LinearInterpolate(A=@n152.B, B=@n152.B, Alpha=@n153.G), Alpha=@n153.R), B=@n152.G, Alpha=@n153.B)
  n155 = ScalarParameter() {defaultValue=0.5, group=Function_Parallax, parameterName=FParallax_Tri_Planner_MaskContrast, sortPriority=5}
  n156 = StaticSwitchParameter(
   True=LinearInterpolate(A=@n17.RGB, B=LinearInterpolate(A=@n84, B=@n98.RGB, Alpha=@n99), Alpha=@n92.Masking Alternat Axies)
   False=LinearInterpolate(
    A=LinearInterpolate(A=@n100.RGB, B=@n114, Alpha=@n115)
    B=LinearInterpolate(A=@n149, B=@n150, Alpha=@n151)
    Alpha=MaterialFunctionCall(In=@n154, Contrast=@n155) {materialFunction=&CheapContrast}.Result
   )
  ) {group=Function_Parallax, parameterName=FParallax_Skew_Projection, sortPriority=-1}
  n157 = Multiply(A=@n156, B=ScalarParameter() {defaultValue=2, group=Function_Parallax, parameterName=FParallax_Edge_Strength, sortPriority=0})
  n158 = Saturate(None=@n5)
  n159 = Multiply(
   A=Multiply(
    A=LinearInterpolate(A=@n156, B=@n157, Alpha=@n158)
    B=ScalarParameter() {defaultValue=0.75, group=Function_Parallax, parameterName=FParallax_Emissive_Amount, sortPriority=0}
   )
   B=MaterialFunctionCall(Value_Morning=Constant() {r=0.5}, Value_Day=Constant() {r=1}, Value_Evening=Constant() {r=0.5}, Value_Night=Constant() {r=0.25}) {materialFunction=&Scalar_Time_of_Day}.Result
  )
  n160 = SetMaterialAttributes!(MaterialAttributes=NamedRerouteUsage(), BaseColor, Roughness, Specular, Metallic, EmissiveColor=Multiply(A=@n15, B=Add(A=@n16, B=@n159))) {attributeSetTypes [
    #1
    #2
    #3
    #4
    #5
   ]}
  n161 = VectorParameter() {group=Function_Parallax, parameterName=FParallax_BaseColor, sortPriority=0, defaultValue=<rgba 0 0 0 1>}
  n162 = SetMaterialAttributes(
   MaterialAttributes=NamedRerouteUsage()
   BaseColor=@n161.RGB
   Roughness=ScalarParameter() {defaultValue=0.5, group=Function_Parallax, parameterName=FParallax_Rougness, sortPriority=0}
   Specular=ScalarParameter() {defaultValue=0.5, group=Function_Parallax, parameterName=FParallax_Specular, sortPriority=0}
   Metallic=ScalarParameter() {group=Function_Parallax, parameterName=FParallax_Metallic, sortPriority=0}
   EmissiveColor=@n159
  ) {attributeSetTypes [
    #1
    #2
    #3
    #4
    #5
   ]}
  n163 = BlendMaterialAttributes(
   A=NamedRerouteUsage()
   B=@n162
   Alpha=LinearInterpolate(
    A=@n15
    B=FunctionInput() {bUsePreviewValueAsDefault=true, inputName=Customization Mask B, inputType=FunctionInput_Scalar, sortPriority=24}
    Alpha=NamedRerouteUsage()
   )
  ) {vertexAttributeBlendType=UseA}
  n164 = FunctionInput(Preview=SetMaterialAttributes!(MaterialAttributes, Normal=TextureSample() {samplerType=SAMPLERTYPE_Normal, texture=/Game/Weapons/FORT_Rifles/T_SCAR_N}.RGB) {attributeSetTypes [#6]}) {bUsePreviewValueAsDefault=true, inputName=Material, inputType=FunctionInput_MaterialAttributes, sortPriority=-20}
  FunctionOutput(None=StaticSwitchParameter(
    True=StaticSwitchParameter(True=@n160, False=@n163) {group=Function_Parallax, parameterName=FParallax_Only Effect Emissive, sortPriority=-1}
    False=@n164
   ) {group=Customization, parameterName=Function_Parallax}) {outputName=Result_1}
  Multiply(A=@n8)
  NamedRerouteDeclaration(None=MaterialFunctionCall(Float4=@n9) {materialFunction=&BreakOutFloat4Components}.A) {name=IsCharacter, nodeColor=<rgba 0.7843 1 0.0442 1>}
  NamedRerouteDeclaration(None=@n13) {name=isVehicle, nodeColor=<rgba 0 1 0.1328 1>}
  n165 = VectorParameter() {group=Function_Parallax_UVS, parameterName=FParallax_UV_Offset_Vehicle_01, defaultValue=<rgba 0 0 0 1>}
  n166 = VectorParameter() {group=Function_Parallax_UVS, parameterName=FParallax_UV_Offset_VehicleSide_01, defaultValue=<rgba 0 0 0 1>}
  n167 = MaterialFunctionCall(Float4=NamedRerouteUsage()) {materialFunction=&BreakOutFloat4Components}
  n168 = VectorParameter() {group=Function_Parallax_UVS, parameterName=FParallax_UV_Offset_Character_01, defaultValue=<rgba 0 0 0 1>}
  n169 = LinearInterpolate(
   A=LinearInterpolate(
    A=VectorParameter() {group=Function_Parallax_UVS, parameterName=FParallax_UV_Offset_Weapon_01, defaultValue=<rgba 0 0 0 1>}.RGB
    B=LinearInterpolate(A=@n165.RGB, B=@n166.RGB, Alpha=@n167.G)
    Alpha=@n167.R
   )
   B=LinearInterpolate(
    A=VectorParameter() {group=Function_Parallax_UVS, parameterName=FParallax_UV_Offset_Mech_01, defaultValue=<rgba 0 0 0 1>}.RGB
    B=@n168.RGB
    Alpha=@n167.A
   )
   Alpha=@n167.B
  )
  n170 = VectorParameter() {group=Function_Parallax_UVS, parameterName=FParallax_UV_Offset_Backpack_01, defaultValue=<rgba 0 0 0 1>}
  n171 = NamedRerouteUsage()
  NamedRerouteDeclaration(None=LinearInterpolate(
    A=LinearInterpolate(A=@n169, B=@n170.RGB, Alpha=@n171)
    B=VectorParameter() {group=Function_Parallax_UVS, parameterName=FParallax_UV_Offset_Pickaxe_01, defaultValue=<rgba 0 0 0 1>}.RGB
    Alpha=NamedRerouteUsage()
   )) {name=isState_offset, nodeColor=<rgba 0 1 0.625 1>}
  n172 = TextureCoordinate()
  n173 = NamedRerouteUsage()
  n174 = LinearInterpolate(
   A=ScalarParameter() {defaultValue=1, group=Function_Parallax, parameterName=FParallax_WarpUVS-01-TileRate Weapon, sortPriority=53}
   B=ScalarParameter() {defaultValue=1, group=Function_Parallax, parameterName=FParallax_WarpUVS-01-TileRate Vehicle, sortPriority=54}
   Alpha=@n173
  )
  n175 = Time() {period=1}
  n176 = VectorParameter() {group=Function_Parallax, parameterName=FParallax_WarpUVS-01-Animated Weapon, sortPriority=57, defaultValue=<rgba 0 0 0 1>}
  n177 = VectorParameter() {group=Function_Parallax, parameterName=FParallax_WarpUVS-01-Animated Vehicle, sortPriority=57, defaultValue=<rgba 0 0 0 1>}
  n178 = ComponentMask(None=LinearInterpolate(A=@n176.RGB, B=@n177.RGB, Alpha=@n173)) {g=true, r=true}
  n179 = TextureSampleParameter2D(UVs=Add(A=Multiply(A=@n172, B=@n174), B=Multiply(A=@n175, B=@n178))) {
   group=Function_Parallax
   parameterName=FParallax_WarpUVS-01-Texture
   sortPriority=52
   texture=&T_BlacksRGB
  }
  n180 = ChannelMaskParameter(None=@n179.RGB) {group=Function_Parallax, parameterName=FParallax_WarpUVS-01_Channel, sortPriority=51}
  n181 = LinearInterpolate(
   A=ScalarParameter() {group=Function_Parallax, parameterName=FParallax_WarpUVS-01-Amount Weapon, sortPriority=55}
   B=ScalarParameter() {group=Function_Parallax, parameterName=FParallax_WarpUVS-01-Amount_Vehicle, sortPriority=56}
   Alpha=@n173
  )
  NamedRerouteDeclaration(None=LinearInterpolate(
    A=Constant()
    B=Multiply(A=@n180.R, B=@n181)
    Alpha=ScalarParameter() {group=Function_Parallax, parameterName=FParallax_WarpUVS-01-ON, sortPriority=56}
   )) {name=Warp_UV_01, nodeColor=<rgba 0 1 0.0625 1>}
  n182 = GetMaterialAttributes(None=@n164) {attributeGetTypes [#5]}
  NamedRerouteDeclaration(None=@n182.MaterialAttributes) {name=RootMateral, nodeColor=<rgba 0.9219 1 0 1>}
  NamedRerouteDeclaration(None=@n182.EmissiveColor) {name=RootEmissive, nodeColor=<rgba 0.9219 1 0 1>}
  n183 = NamedRerouteDeclaration(None=FunctionInput() {bUsePreviewValueAsDefault=true, inputName=IsStatsB, inputType=FunctionInput_Vector2, sortPriority=78}) {name=Is_StatsB, nodeColor=<rgba 0 0.9253 1 1>}
  NamedRerouteDeclaration(None=ComponentMask(None=@n183) {g=true}) {name=IsPickaxe, nodeColor=<rgba 1 0 0.5625 1>}
  NamedRerouteDeclaration(None=ComponentMask(None=@n183) {r=true}) {name=IsBackpack, nodeColor=<rgba 0.2656 0 1 1>}
  NamedRerouteDeclaration(None=@n9) {name=IsState, nodeColor=<rgba 0 0.0859 1 1>}
 ]}
