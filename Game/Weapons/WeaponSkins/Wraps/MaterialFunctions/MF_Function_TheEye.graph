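# Material function for weapon wraps that layers a projected, glowing eye and an accent material onto an input material, toggled by the Function_TheEye static switch.
# Side-projection atlas samples build the eye and four surrounding strips, with scale, rotation and offset parameters chosen per object type (weapon, vehicle, mech, character, backpack, pickaxe).
# Lerps animate parts and pupil radius, alongside a Blueprint-timestamped TimeBP ramp.
# Input masks and a normal-based directional mask support blending.
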MaterialFunction /Game/Weapons/WeaponSkins/Wraps/MaterialFunctions/MF_Function_TheEye
graph {nodes [
  Constant2Vector() {g=-0.02}
  Constant4Vector()
  Constant() {r=0.1}
  Floor()
  n1 = Time()
  n2 = Multiply(A=@n1)
  Frac(None=@n2)
  n3 = BlendMaterialAttributes(A=NamedRerouteUsage(), B=NamedRerouteUsage(), Alpha=NamedRerouteUsage()) {vertexAttributeBlendType=UseA}
  n4 = NamedRerouteUsage()
  n5 = TextureSample() {samplerType=SAMPLERTYPE_Masks, texture=/Game/Weapons/FORT_Rifles/Textures/T_Scar_CM}
  n6 = FunctionInput(Preview=@n5.R) {bUsePreviewValueAsDefault=true, inputName=Mask Primary, inputType=FunctionInput_Scalar, sortPriority=24, previewValue=<1 0 0 1>}
  n7 = FunctionInput(Preview=@n5.B) {bUsePreviewValueAsDefault=true, inputName=Mask Secondary, inputType=FunctionInput_Scalar, sortPriority=24, previewValue=<1 0 0 1>}
  n8 = Multiply(A=Saturate(None=Add(A=@n6, B=@n7)), B=NamedRerouteUsage())
  n9 = TextureSample() {samplerType=SAMPLERTYPE_Normal, texture=/Game/Weapons/FORT_Rifles/T_SCAR_N}
  n10 = Constant3Vector() {constant=<rgba 1 0 0 0>}
  n11 = GetMaterialAttributes(None=FunctionInput(Preview=SetMaterialAttributes!(MaterialAttributes, Normal=@n9.RGB, BaseColor=@n10) {attributeSetTypes [#1, #2]}) {bUsePreviewValueAsDefault=true, inputName=Material, inputType=FunctionInput_MaterialAttributes}) {attributeGetTypes [
    #2
    #3
    #1
    #4
   ]}
  FunctionOutput(None=StaticSwitchParameter(
    True=BlendMaterialAttributes(A=@n3, B=@n4, Alpha=@n8) {vertexAttributeBlendType=UseA}
    False=NamedRerouteDeclaration(None=@n11.MaterialAttributes) {name=Root_Material, nodeColor=<rgba 0.7812 1 0 1>}
   ) {desc=TopSide Projection Base - Emissive pan Mask, group=Customization, parameterName=Function_TheEye})
  n12 = MaterialFunctionCall(Float4=NamedRerouteUsage()) {materialFunction=&BreakOutFloat4Components}
  n13 = NamedRerouteUsage()
  MaterialFunctionCall!(
   Normal_Map=NamedRerouteUsage()
   Caculated_Bounds=NamedRerouteUsage()
   Pre_Skinned_Normals=NamedRerouteUsage()
   Is Vehicle=@n12.R
   Is VehicleSide=@n12.R
   Is Mech=@n12.R
   UV_Rotation=Add(A=Constant() {r=0.34}, B=NamedRerouteUsage())
   Uvs_Offset=NamedRerouteUsage()
   Uvs_Scale=@n13
   Atlas_offset_location=Constant2Vector() {g=0.5}
   "Atlas_Row\\Colums Amount"
  ) {materialFunction=&/MF_Projection_Side_Nomal_Atlas}
  n14 = NamedRerouteUsage()
  MaterialFunctionCall(UVs=@n14, Texture Scale=@n13) {materialFunction=&ScaleUVsByCenter}
  n15 = Divide(B=Constant2Vector() {g=2, r=2})
  n16 = TextureSample(UVs=Add(A=Multiply(A=@n14, B=@n15), B=Constant2Vector() {g=0.5}), Tex=NamedRerouteUsage()) {samplerSource=SSM_Clamp_WorldGroupSettings}
  Multiply(A=@n16.A)
  NamedRerouteDeclaration(None=TextureObjectParameter() {group=Func_TheEye, parameterName=TEye_Tex__BC, texture=&T_BlacksRGB}) {name=Tex_BaseColor, nodeColor=<rgba 0 0.9062 1 1>}
  NamedRerouteDeclaration(None=TextureObjectParameter() {group=Func_TheEye, parameterName=TEye_Tex__MSK, texture=&T_BlacksRGB}) {name=Tex_Mask, nodeColor=<rgba 0 0.9062 1 1>}
  n17 = MaterialFunctionCall(Float4=NamedRerouteUsage()) {materialFunction=&BreakOutFloat4Components}
  n18 = Constant() {r=0.34}
  n19 = NamedRerouteUsage()
  n20 = Constant2Vector() {g=-0.055}
  n21 = Constant2Vector() {g=0.055}
  n22 = MaterialFunctionCall(
   Normal_Map=NamedRerouteUsage()
   Caculated_Bounds=NamedRerouteUsage()
   Pre_Skinned_Normals=NamedRerouteUsage()
   Is Vehicle=@n17.R
   Is VehicleSide=@n17.R
   Is Mech=@n17.R
   UV_Rotation=Add(A=LinearInterpolate(A=@n18, Alpha=@n19) {constB=0.5}, B=NamedRerouteUsage())
   Uvs_Offset=NamedRerouteUsage()
   Uvs_Scale=Multiply(A=NamedRerouteUsage(), B=LinearInterpolate(Alpha=@n19) {constA=1, constB=1.05})
   Atlas_offset_location=Add(A=Constant2Vector() {g=0.5}, B=LinearInterpolate(A=@n20, B=@n21, Alpha=@n19)) {constB=2}
  ) {materialFunction=&MF_Projection_Side_Nomal_Atlas}
  n23 = TextureSample(UVs=@n22.UVs, Tex=NamedRerouteUsage()) {samplerSource=SSM_Wrap_WorldGroupSettings, samplerType=SAMPLERTYPE_Masks}
  n24 = MaterialFunctionCall(Float4=NamedRerouteUsage()) {materialFunction=&BreakOutFloat4Components}
  n25 = Constant() {r=0.12}
  n26 = NamedRerouteUsage()
  n27 = MaterialFunctionCall(
   Normal_Map=NamedRerouteUsage()
   Caculated_Bounds=NamedRerouteUsage()
   Pre_Skinned_Normals=NamedRerouteUsage()
   Is Vehicle=@n24.R
   Is VehicleSide=@n24.R
   Is Mech=@n24.R
   UV_Rotation=Add(A=LinearInterpolate(A=@n25, Alpha=@n26) {constB=0.5}, B=NamedRerouteUsage())
   Uvs_Offset=NamedRerouteUsage()
   Uvs_Scale=NamedRerouteUsage()
   Atlas_offset_location=Add(A=Constant2Vector() {g=0.5}, B=Constant2Vector() {g=-0.07, r=0.02}) {constB=2}
  ) {materialFunction=&MF_Projection_Side_Nomal_Atlas}
  n28 = TextureSample(UVs=@n27.UVs, Tex=NamedRerouteUsage()) {samplerSource=SSM_Wrap_WorldGroupSettings, samplerType=SAMPLERTYPE_Masks}
  n29 = MaterialFunctionCall(Float4=NamedRerouteUsage()) {materialFunction=&BreakOutFloat4Components}
  n30 = Constant() {r=0.9}
  n31 = NamedRerouteUsage()
  n32 = MaterialFunctionCall(
   Normal_Map=NamedRerouteUsage()
   Caculated_Bounds=NamedRerouteUsage()
   Pre_Skinned_Normals=NamedRerouteUsage()
   Is Vehicle=@n29.R
   Is VehicleSide=@n29.R
   Is Mech=@n29.R
   UV_Rotation=Add(A=NamedRerouteUsage(), B=LinearInterpolate(A=@n30, Alpha=@n31) {constB=0.5})
   Uvs_Offset=NamedRerouteUsage()
   Uvs_Scale=NamedRerouteUsage()
   Atlas_offset_location=Add(A=Constant2Vector() {g=0.5}, B=Constant2Vector() {g=-0.055})
  ) {materialFunction=&MF_Projection_Side_Nomal_Atlas}
  n33 = TextureSample(UVs=@n32.UVs, Tex=NamedRerouteUsage()) {samplerSource=SSM_Wrap_WorldGroupSettings, samplerType=SAMPLERTYPE_Masks}
  n34 = MaterialFunctionCall(Float4=NamedRerouteUsage()) {materialFunction=&BreakOutFloat4Components}
  n35 = Constant() {r=0.63}
  n36 = NamedRerouteUsage()
  n37 = Constant2Vector() {g=-0.02}
  n38 = Constant2Vector() {g=-0.1}
  n39 = MaterialFunctionCall(
   Normal_Map=NamedRerouteUsage()
   Caculated_Bounds=NamedRerouteUsage()
   Pre_Skinned_Normals=NamedRerouteUsage()
   Is Vehicle=@n34.R
   Is VehicleSide=@n34.R
   Is Mech=@n34.R
   UV_Rotation=Add(A=NamedRerouteUsage(), B=LinearInterpolate(A=@n35, Alpha=@n36) {constB=0})
   Uvs_Offset=NamedRerouteUsage()
   Uvs_Scale=NamedRerouteUsage()
   Atlas_offset_location=Add(A=Constant2Vector() {g=0.5}, B=LinearInterpolate(A=@n37, B=@n38, Alpha=@n36))
  ) {materialFunction=&MF_Projection_Side_Nomal_Atlas}
  n40 = TextureSample(UVs=@n39.UVs, Tex=NamedRerouteUsage()) {samplerSource=SSM_Wrap_WorldGroupSettings, samplerType=SAMPLERTYPE_Masks}
  n41 = Add(A=Add(A=@n23.A, B=@n28.A), B=Add(A=@n33.A, B=@n40.A))
  NamedRerouteDeclaration(None=Multiply(A=@n16.A, B=Saturate(None=@n41))) {name=Strips_Mask, nodeColor=<rgba 0 0.8828 1 1>}
  n42 = GetMaterialAttributes(None=FunctionInput() {bUsePreviewValueAsDefault=true, inputName=Material_Root, inputType=FunctionInput_MaterialAttributes}) {attributeGetTypes [
    #2
    #3
    #1
    #4
   ]}
  NamedRerouteDeclaration(None=@n42.Normal) {name=Root_Normal_Master, nodeColor=<rgba 0.4062 0 1 1>}
  n43 = NamedRerouteUsage()
  n44 = MaterialFunctionCall(Float4=NamedRerouteUsage()) {materialFunction=&BreakOutFloat4Components}
  n45 = NamedRerouteUsage()
  n46 = NamedRerouteUsage()
  n47 = NamedRerouteUsage()
  n48 = MaterialFunctionCall(
   Normal_Map=NamedRerouteUsage()
   Caculated_Bounds=NamedRerouteUsage()
   Pre_Skinned_Normals=NamedRerouteUsage()
   Is Vehicle=@n44.R
   Is VehicleSide=@n44.R
   Is Mech=@n44.R
   UV_Rotation=@n45
   Uvs_Offset=@n46
   Uvs_Scale=@n47
  ) {materialFunction=&MF_Projection_Side_Nomal_Atlas}
  n49 = TextureSample(UVs=@n48.UVs, Tex=NamedRerouteUsage()) {samplerSource=SSM_Wrap_WorldGroupSettings, samplerType=SAMPLERTYPE_Masks}
  n50 = SphereMask(
   A=ComponentMask(None=@n48.UVs) {g=true}
   B=Constant2Vector() {g=0.25, r=0.25}
   Radius=LinearInterpolate(Alpha=NamedRerouteUsage()) {constAlpha=1, constB=0.09}
  ) {attenuationRadius=0.25}
  n51 = OneMinus(None=@n50)
  n52 = Multiply(A=@n49.RGB, B=@n51)
  n53 = TextureSample(UVs=@n48.UVs, Tex=NamedRerouteUsage()) {samplerSource=SSM_Clamp_WorldGroupSettings}
  n54 = MaterialFunctionCall(Float4=NamedRerouteUsage()) {materialFunction=&BreakOutFloat4Components}
  n55 = NamedRerouteUsage()
  n56 = MaterialFunctionCall(
   Normal_Map=NamedRerouteUsage()
   Caculated_Bounds=NamedRerouteUsage()
   Pre_Skinned_Normals=NamedRerouteUsage()
   Is Vehicle=@n54.R
   Is VehicleSide=@n54.R
   Is Mech=@n54.R
   UV_Rotation=Add(A=@n45, B=LinearInterpolate(Alpha=@n55) {constB=-0.5})
   Uvs_Offset=@n46
   Uvs_Scale=@n47
   Atlas_offset_location=Constant2Vector() {r=0.5}
  ) {materialFunction=&MF_Projection_Side_Nomal_Atlas}
  n57 = TextureSample(UVs=@n56.UVs, Tex=NamedRerouteUsage()) {samplerSource=SSM_Wrap_WorldGroupSettings, samplerType=SAMPLERTYPE_Masks}
  n58 = LinearInterpolate(A=LinearInterpolate(A=LinearInterpolate(A=@n49.RGB, B=@n52, Alpha=@n53.A), B=@n57.RGB, Alpha=@n57.RGB), B=NamedRerouteUsage(), Alpha=@n43)
  n59 = ScalarParameter() {defaultValue=0.2, group=Func_TheEye, parameterName=TEye_Roughtness_Inner, sortPriority=1}
  n60 = ChannelMaskParameter(None=@n58) {group=Func_TheEye, maskChannel=Green, parameterName=TEye_Rougness_Source, sortPriority=0, defaultValue=<rgba 0 1 0 0>}
  n61 = Saturate(None=LinearInterpolate(
    A=ScalarParameter() {defaultValue=1, group=Func_TheEye, parameterName=TEye_Roughtness_Outer_Low, sortPriority=1}
    B=ScalarParameter() {defaultValue=0.35, group=Func_TheEye, parameterName=TEye_Roughtness_Outer_High, sortPriority=1}
    Alpha=@n60.R
   ))
  n62 = OneMinus(None=@n53.A)
  n63 = NamedRerouteUsage()
  n64 = Saturate(None=Add(A=Add(A=@n62, B=@n51), B=@n63))
  n65 = Multiply(A=@n64, B=NamedRerouteUsage())
  n66 = Multiply(A=Multiply(A=@n53.A, B=@n50), B=VectorParameter() {group=Func_TheEye, parameterName=TEye_Glow_Color, defaultValue=<rgba 0 10 0.999 1>}.RGB)
  n67 = OneMinus(None=@n63)
  n68 = MaterialFunctionCall(UVs=@n48.UV_No_offset, Texture Scale=NamedRerouteUsage()) {materialFunction=&ScaleUVsByCenter}
  n69 = Multiply(
   A=TextureSample(UVs=NamedRerouteDeclaration(None=@n68.UVs) {name=Uvs_No offset, nodeColor=<rgba 1 0 0.375 1>}, Tex=NamedRerouteUsage()) {samplerSource=SSM_Clamp_WorldGroupSettings}.B
   B=@n49.A
  )
  n70 = LinearInterpolate(A=LinearInterpolate(A=@n48.Raw_Normal, B=@n56.Raw_Normal, Alpha=@n57.A), B=Constant3Vector() {constant=<rgba 0 0 1 0>}, Alpha=@n43)
  n71 = Constant3Vector() {constant=<rgba 1 1 0 0>}
  n72 = SetMaterialAttributes!(
   MaterialAttributes=NamedRerouteUsage()
   BaseColor=@n58
   Roughness=Saturate(None=LinearInterpolate(A=@n59, B=@n61, Alpha=@n64))
   Specular=LinearInterpolate(
    A=ScalarParameter() {defaultValue=0.5, group=Func_TheEye, parameterName=TEye_Spec_Inner, sortPriority=1}
    B=ScalarParameter() {defaultValue=0.5, group=Func_TheEye, parameterName=TEye_Spec_Outer, sortPriority=1}
    Alpha=@n64
   )
   Metallic=Round(None=LinearInterpolate(Alpha=@n65) {constA=-2, constB=0.75})
   EmissiveColor=Multiply(A=NamedRerouteDeclaration(None=@n69) {name=TheEye_Mask, nodeColor=<rgba 1 0 0.1172 1>}, B=Multiply(A=@n66, B=@n67))
   Normal=Add(A=Multiply(A=@n70, B=@n71), B=Constant3Vector() {constant=<rgba 0 0 1 0>})
  ) {attributeSetTypes [
    #2
    #3
    #5
    #4
    #6
    #1
   ]}
  NamedRerouteDeclaration(None=@n72) {name=Material_ThEye, nodeColor=<rgba 0.3828 1 0 1>}
  NamedRerouteDeclaration(None=FunctionInput(Preview=Constant4Vector() {constant=<rgba 0 0 0 0.2>}.RGBA) {bUsePreviewValueAsDefault=true, inputName=Object-SafeZone-Front, inputType=FunctionInput_Vector4, sortPriority=102}) {name=SafeZone_Frnt, nodeColor=<rgba 0.9922 1 0 1>}
  n73 = ScalarParameter() {defaultValue=1, group=Func_TheEye_UVS, parameterName=TEye_ScaleBack_Weapon}
  n74 = MaterialFunctionCall(Float4=NamedRerouteUsage()) {materialFunction=&BreakOutFloat4Components}
  n75 = LinearInterpolate(
   A=ScalarParameter() {defaultValue=1, group=Func_TheEye_UVS, parameterName=TEye_ScaleBack_Vehicle}
   B=ScalarParameter() {defaultValue=1, group=Func_TheEye_UVS, parameterName=TEye_ScaleBack_Vehicle_Side}
   Alpha=@n74.G
  )
  n76 = ScalarParameter() {defaultValue=1, group=Func_TheEye_UVS, parameterName=TEye_ScaleBack_Mech}
  n77 = ScalarParameter() {defaultValue=1, group=Func_TheEye_UVS, parameterName=TEye_ScaleBack_Character}
  n78 = MaterialFunctionCall(Float2=NamedRerouteUsage()) {materialFunction=&BreakOutFloat2Components}
  n79 = LinearInterpolate(
   A=LinearInterpolate(A=LinearInterpolate(A=@n73, B=@n75, Alpha=@n74.R), B=LinearInterpolate(A=@n76, B=@n77, Alpha=@n74.A), Alpha=@n74.B)
   B=ScalarParameter() {defaultValue=1, group=Func_TheEye_UVS, parameterName=TEye_ScaleBack_Backpack}
   Alpha=@n78.R
  )
  n80 = ScalarParameter() {defaultValue=1, group=Func_TheEye_UVS, parameterName=TEye_ScaleBack_Pickaxe}
  n81 = NamedRerouteUsage()
  NamedRerouteDeclaration(None=Multiply(A=LinearInterpolate(A=@n79, B=@n80, Alpha=@n78.G), B=ComponentMask(None=@n81) {a=true})) {name=UV_Scale, nodeColor=<rgba 0.4062 0 1 1>}
  n82 = ScalarParameter() {group=Func_TheEye_UVS, parameterName=TEye_Rotation_Vehicle}
  n83 = ScalarParameter() {group=Func_TheEye_UVS, parameterName=TEye_Rotation_Vehicle_Side}
  n84 = MaterialFunctionCall(Float4=NamedRerouteUsage()) {materialFunction=&BreakOutFloat4Components}
  n85 = LinearInterpolate(
   A=LinearInterpolate(
    A=ScalarParameter() {group=Func_TheEye_UVS, parameterName=TEye_Rotation_Weapon}
    B=LinearInterpolate(A=@n82, B=@n83, Alpha=@n84.G)
    Alpha=@n84.R
   )
   B=LinearInterpolate(
    A=ScalarParameter() {group=Func_TheEye_UVS, parameterName=TEye_Rotation_Mech}
    B=ScalarParameter() {group=Func_TheEye_UVS, parameterName=TEye_Rotation_Character}
    Alpha=@n84.A
   )
   Alpha=@n84.B
  )
  n86 = ScalarParameter() {group=Func_TheEye_UVS, parameterName=TEye_Rotation_Backpack}
  n87 = MaterialFunctionCall(Float2=NamedRerouteUsage()) {materialFunction=&BreakOutFloat2Components}
  NamedRerouteDeclaration(None=LinearInterpolate(
    A=LinearInterpolate(A=@n85, B=@n86, Alpha=@n87.R)
    B=ScalarParameter() {group=Func_TheEye_UVS, parameterName=TEye_Rotation_Pickaxe}
    Alpha=@n87.G
   )) {name=UV_Rotation, nodeColor=<rgba 0.4062 0 1 1>}
  n88 = VectorParameter() {group=Func_TheEye_UVS, parameterName=TEye_Offset_Weapon_01, defaultValue=<rgba 0 0 0 1>}
  n89 = MaterialFunctionCall(Float4=NamedRerouteUsage()) {materialFunction=&BreakOutFloat4Components}
  n90 = LinearInterpolate(
   A=VectorParameter() {group=Func_TheEye_UVS, parameterName=TEye_Offset_Vehicle_01, defaultValue=<rgba 0 0 0 1>}.RGB
   B=VectorParameter() {group=Func_TheEye_UVS, parameterName=TEye_Offset_VehicleSide_01, defaultValue=<rgba 0 0 0 1>}.RGB
   Alpha=@n89.G
  )
  n91 = VectorParameter() {group=Func_TheEye_UVS, parameterName=TEye_Offset_Mech_01, defaultValue=<rgba 0 0 0 1>}
  n92 = VectorParameter() {group=Func_TheEye_UVS, parameterName=SLST_Ship_Offset_Character_01, defaultValue=<rgba 0 0 0 1>}
  n93 = MaterialFunctionCall(Float2=NamedRerouteUsage()) {materialFunction=&BreakOutFloat2Components}
  n94 = LinearInterpolate(
   A=LinearInterpolate(A=LinearInterpolate(A=@n88.RGB, B=@n90, Alpha=@n89.R), B=LinearInterpolate(A=@n91.RGB, B=@n92.RGB, Alpha=@n89.A), Alpha=@n89.B)
   B=VectorParameter() {group=Func_TheEye_UVS, parameterName=TEye_Offset_Backpack, defaultValue=<rgba 0 0 0 1>}.RGB
   Alpha=@n93.R
  )
  n95 = VectorParameter() {group=Func_TheEye_UVS, parameterName=TEye_Offset_Pickaxe_01, defaultValue=<rgba 0 0 0 1>}
  n96 = NamedRerouteUsage()
  NamedRerouteDeclaration(None=Add(A=LinearInterpolate(A=@n94, B=@n95.RGB, Alpha=@n93.G), B=ComponentMask(None=@n96) {b=true, g=true, r=true})) {name=UV_Offset, nodeColor=<rgba 1 0 0.8437 1>}
  NamedRerouteDeclaration(None=@n7) {name=Mask_Secondary, nodeColor=<rgba 1 0 0.6562 1>}
  n97 = TextureObjectParameter() {group=Func_TheEye, parameterName=TEye_Tex__N, samplerType=SAMPLERTYPE_Normal, texture=&FlatNormalBlue}
  NamedRerouteDeclaration(None=@n97) {name=Tex_Normal, nodeColor=<rgba 0 0.9062 1 1>}
  n98 = VectorParameter() {group=Func_TheEye_Accent, parameterName=TEye_Accent_BaseColor, defaultValue=<rgba 0.0394 0.974 0.0727 1>}
  n99 = SetMaterialAttributes!(
   MaterialAttributes=NamedRerouteUsage()
   BaseColor=@n98.RGB
   Metallic=ScalarParameter() {group=Func_TheEye_Accent, parameterName=TEye_Accent_Metallic}
   Specular=ScalarParameter() {defaultValue=0.5, group=Func_TheEye_Accent, parameterName=TEye_Accent_Spec}
   Roughness=ScalarParameter() {defaultValue=0.5, group=Func_TheEye_Accent, parameterName=TEye_Accent_Roughness}
   Normal=NamedRerouteUsage()
  ) {attributeSetTypes [
    #2
    #4
    #5
    #3
    #1
   ]}
  NamedRerouteDeclaration(None=@n99) {name=Accent_Material, nodeColor=<rgba 0 0.8359 1 1>}
  n100 = Divide(
   A=Subtract(A=@n1, B=ScalarParameter() {defaultValue=-9999999, group=Func_TheEye, parameterName=TEye_TimeFrom_BP, sortPriority=100})
   B=ScalarParameter() {defaultValue=0.25, group=Func_TheEye, parameterName=TEye_TimeFrom_BP_Duration, sortPriority=100}
  )
  NamedRerouteDeclaration(None=LinearInterpolate(
    A=ScalarParameter() {defaultValue=1, group=Func_TheEye, parameterName=TEye_Time_End, sortPriority=120}
    B=ScalarParameter() {group=Func_TheEye, parameterName=TEye_Time_Start, sortPriority=120}
    Alpha=Saturate(None=@n100)
   )) {name=TimeBP, nodeColor=<rgba 0.7812 1 0 1>}
  n101 = MaterialFunctionCall(Color=Abs(None=NamedRerouteUsage())) {materialFunction=&SplitComponents}
  n102 = MaterialFunctionCall(Float4=NamedRerouteUsage()) {materialFunction=&BreakOutFloat4Components}
  n103 = LinearInterpolate(A=@n101.R, B=LinearInterpolate(A=@n101.B, B=@n101.B, Alpha=@n102.G), Alpha=@n102.R)
  NamedRerouteDeclaration(None=MaterialFunctionCall(
    In=LinearInterpolate(A=@n103, B=@n101.G, Alpha=@n102.B)
    Contrast=ScalarParameter() {defaultValue=1, group=Func_TheEye, parameterName=TEye_TriMask_Contrast, sortPriority=-1}
   ) {materialFunction=&CheapContrast}.Result) {name=Tri_Mask, nodeColor=<rgba 1 0.4688 0 1>}
  n104 = LinearInterpolate(A=@n39.Raw_Normal, B=@n32.Raw_Normal, Alpha=@n33.A)
  NamedRerouteDeclaration(None=LinearInterpolate(A=LinearInterpolate(A=@n104, B=@n27.Raw_Normal, Alpha=@n28.A), B=@n22.Raw_Normal, Alpha=@n23.A)) {name=Strips_Normal, nodeColor=<rgba 0 0.8828 1 1>}
  n105 = FunctionInput(Preview=@n5.G) {bUsePreviewValueAsDefault=true, inputName=Mask Crevise, inputType=FunctionInput_Scalar, sortPriority=24, previewValue=<1 0 0 1>}
  NamedRerouteDeclaration(None=@n105) {name=Crevise_map, nodeColor=<rgba 1 0.6797 0 1>}
  NamedRerouteDeclaration(None=FunctionInput(Preview=Constant4Vector().RGBA) {bUsePreviewValueAsDefault=true, inputName=IsState, inputType=FunctionInput_Vector4, sortPriority=4}) {name=IsState, nodeColor=<rgba 0.1497 1 0.1633 1>}
  NamedRerouteDeclaration(None=FunctionInput(Preview=Constant2Vector()) {bUsePreviewValueAsDefault=true, inputName=IsState_B, inputType=FunctionInput_Vector2, sortPriority=4}) {name=IsState_B, nodeColor=<rgba 0.1497 1 0.1633 1>}
  n106 = PreSkinnedLocalBounds()
  n107 = Subtract(A=VertexInterpolator(VS=PreSkinnedPosition()).PS, B=@n106.Min)
  n108 = MaterialFunctionCall(Color=Subtract(A=@n106.Max, B=@n106.Max)) {materialFunction=&SplitComponents}
  n109 = Max(A=Max(A=@n108.R, B=@n108.R), B=@n108.B)
  NamedRerouteDeclaration(None=FunctionInput(Preview=Divide(A=@n107, B=@n109)) {bUsePreviewValueAsDefault=true, inputName=Calcated_Bounds, sortPriority=11}) {name=Caclcated_Bounds, nodeColor=<rgba 0 0.9297 1 1>}
  n110 = LinearInterpolate(A=@n40.RGB, B=@n33.RGB, Alpha=@n33.RGB)
  NamedRerouteDeclaration(None=LinearInterpolate(A=LinearInterpolate(A=@n110, B=@n28.RGB, Alpha=@n28.RGB), B=@n23.RGB, Alpha=@n23.RGB)) {name=Strips_BC, nodeColor=<rgba 0 0.8828 1 1>}
  n111 = PreSkinnedNormal()
  NamedRerouteDeclaration(None=FunctionInput(Preview=VertexInterpolator(VS=@n111).PS) {bUsePreviewValueAsDefault=true, inputName=Pre-Skinned Local Normal, sortPriority=10}) {name=PreSkin_Normal, nodeColor=<rgba 0.8281 1 0 1>}
  NamedRerouteDeclaration(None=@n11.Normal) {name=Root_Normal, nodeColor=<rgba 0.4062 0 1 1>}
  NamedRerouteUsage()
  NamedRerouteUsage()
  Saturate(None=Fmod(A=@n2, B=Constant() {r=1.5}))
 ]}
