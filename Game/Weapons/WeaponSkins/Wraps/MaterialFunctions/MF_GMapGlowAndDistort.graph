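# Material function for weapon wraps that adds a curve-colored emissive glow, blended in by a mask channel behind a static switch.
# Adjusted curvature and GMap values sum to index a curve atlas row, whose color is modulated by noise projected from pre-skinned position and normal.
# isVehicle and isMech inputs select noise tiling and minimum brightness.
# Despite the name, the distortion and panning nodes do not reach the output.
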MaterialFunction /Game/Weapons/WeaponSkins/Wraps/MaterialFunctions/MF_GMapGlowAndDistort
graph {nodes [
  n1 = FunctionInput(Preview=VertexInterpolator(VS=PreSkinnedPosition()).PS) {bUsePreviewValueAsDefault=true, inputName=PreskinnedPosition}
  n2 = VectorParameter() {group=GMap_Glow, parameterName=GG_WeaponDistortionPanSpeedAndDirection, defaultValue=<rgba 0.75 0.75 0.75 1>}
  n3 = Constant()
  n4 = FunctionInput(Preview=@n3) {bUsePreviewValueAsDefault=true, inputName=isMech, inputType=FunctionInput_Scalar}
  n5 = FunctionInput(Preview=@n3) {bUsePreviewValueAsDefault=true, inputName=isVehicle, inputType=FunctionInput_Scalar}
  n6 = TextureObjectParameter() {group=GMap_Glow, parameterName=DIstortNoise, texture=/Game/Characters/Player/Male/Medium/Bodies/M_MED_Viper/Textures/Glitch/T_SoftNoise}
  n7 = MaterialFunctionCall!(
   WeaponDistortionPanSpeedAndDirection=@n2.RGB
   WeaponDistortionTiling=VectorParameter() {group=GMap_Glow, parameterName=GG_WeaponDistortionTiling, defaultValue=<rgba 32 32 32 1>}.RGB
   VehicleDistortionTiling=VectorParameter() {group=GMap_Glow, parameterName=GG_VehicleDistortionTiling, defaultValue=<rgba 12 12 12 1>}.RGB
   MechDistortionTiling=VectorParameter() {group=GMap_Glow, parameterName=GG_MechDistortionTiling, defaultValue=<rgba 24 24 24 1>}.RGB
   isMech=@n4
   isVehicle=@n5
   DistortionNoise=@n6
   VehicleDistortionPanSpeedAndDirection=VectorParameter() {group=GMap_Glow, parameterName=GG_VehicleDistortionPanSpeedAndDirection, defaultValue=<rgba 0 0 0 1>}.RGB
   MechDistortionPanSpeedAndDirection=VectorParameter() {group=GMap_Glow, parameterName=GG_MechDistortionPanSpeedAndDirection, defaultValue=<rgba 0 0 0 1>}.RGB
   WeaponDistortionStrength=VectorParameter() {group=GMap_Glow, parameterName=GG_WeaponDistortionStrength, defaultValue=<rgba 2 2 2 1>}.RGB
   VehicleDistortionStrength=VectorParameter() {group=GMap_Glow, parameterName=GG_VehicleDistortionStrength, defaultValue=<rgba 0.1 0.1 0.1 1>}.RGB
   MechDistortionStrength=VectorParameter() {group=GMap_Glow, parameterName=GG_MechDistortionStrength, defaultValue=<rgba 0.1 0.1 0.1 1>}.RGB
  ) {materialFunction=&/WorldAlignedDistortion_Wraps}
  n8 = LinearInterpolate(
   A=VectorParameter() {group=GMap_Glow, parameterName=GG_WeaponNoisePanSpeedAndDIrection, defaultValue=<rgba 5 5 5 1>}.RGB
   B=VectorParameter() {group=GMap_Glow, parameterName=GG_VehicleNoisePanSpeedAndDirection, defaultValue=<rgba 0 0 0 1>}.RGB
   Alpha=@n5
  )
  n9 = VectorParameter() {group=GMap_Glow, parameterName=GG_MechNoisePanSpeedAndDirection, defaultValue=<rgba 0 0 0 1>}
  Add(A=Add(B=Add(A=@n1, B=@n7.Result)), B=Multiply(A=LinearInterpolate(A=@n8, B=@n9.RGB, Alpha=@n4), B=Time()))
  ConstantBiasScale(None=@n7.Result)
  Constant()
  FunctionInput(Preview=PreSkinnedLocalBounds().Max) {bUsePreviewValueAsDefault=true, inputName=PreskinnedBounds}
  n10 = GetMaterialAttributes(None=FunctionInput() {inputName=Material, inputType=FunctionInput_MaterialAttributes}) {attributeGetTypes [#1]}
  n11 = TextureSample() {samplerType=SAMPLERTYPE_Masks, texture=/Game/Weapons/FORT_Rifles/Textures/T_Scar_CM}
  n12 = FunctionInput(Preview=@n11.G) {bUsePreviewValueAsDefault=true, inputName=Curvature, inputType=FunctionInput_Scalar}
  n13 = Subtract(
   A=LinearInterpolate(A=@n12, B=OneMinus(None=@n12), Alpha=ScalarParameter() {defaultValue=1, group=GMap_Glow, parameterName=GG_FlipCurvature})
   B=ScalarParameter() {defaultValue=0.5, group=GMap_Glow, parameterName=GG_CurvatureSpread}
  )
  n14 = ScalarParameter() {defaultValue=2, group=GMap_Glow, parameterName=GG_CurvatureStrength}
  n15 = Multiply(A=Clamp(None=Multiply(A=@n13, B=@n14)), B=ScalarParameter() {defaultValue=1, group=GMap_Glow, parameterName=GG_UseCurvature})
  n16 = FunctionInput(Preview=@n11.A) {bUsePreviewValueAsDefault=true, inputName=GMap, inputType=FunctionInput_Scalar}
  n17 = OneMinus(None=@n16)
  n18 = ScalarParameter() {defaultValue=1, group=GMap_Glow, parameterName=GG_Flip_Gmap}
  n19 = Multiply(
   A=Subtract(A=LinearInterpolate(A=@n16, B=@n17, Alpha=@n18), B=ScalarParameter() {defaultValue=0.25, group=GMap_Glow, parameterName=GG_GMapSpread})
   B=ScalarParameter() {defaultValue=2, group=GMap_Glow, parameterName=GG_GMap_Strength}
  )
  n20 = Multiply(A=Clamp(None=Power(Base=@n19)), B=ScalarParameter() {defaultValue=1, group=GMap_Glow, parameterName=GG_UseGMap})
  n21 = CurveAtlasRowParameter(CurveTime=Clamp(None=Add(A=@n15, B=@n20))) {
   atlas=&AthenaCurveAtlas01
   curve=/Game/Weapons/WeaponSkins/Wraps/Textures/BoneSNake/BoneSnakeWrapColor
   defaultValue=151
   group=GMap_Glow
   parameterName=GlowColorCurve
  }
  n22 = VectorParameter() {group=GMap_Glow, parameterName=GG_WeaponNoiseTiling, defaultValue=<rgba 24 24 24 1>}
  n23 = VectorParameter() {group=GMap_Glow, parameterName=GG_VehicleNoiseTiling, defaultValue=<rgba 0 0 0 1>}
  n24 = PreSkinnedNormal()
  n25 = MaterialFunctionCall(
   TextureObject=@n6
   TextureSize=LinearInterpolate(
    A=LinearInterpolate(A=@n22.RGB, B=@n23.RGB, Alpha=@n5)
    B=VectorParameter() {group=GMap_Glow, parameterName=GG_MechNoiseTiling, defaultValue=<rgba 0 0 0 1>}.RGB
    Alpha=@n4
   )
   WorldPosition=@n1
   World Space Normal=FunctionInput(Preview=VertexInterpolator(VS=@n24).PS) {bUsePreviewValueAsDefault=true, inputName=PreskinnedNormal}
  ) {materialFunction=&WorldAlignedTexture}
  n26 = LinearInterpolate(
   A=ScalarParameter() {defaultValue=0.5, group=GMap_Glow, parameterName=WeaponEmissiveVariationNoiseStrength}
   B=ScalarParameter() {defaultValue=0.5, group=GMap_Glow, parameterName=VehicleEmissiveVariationNoiseStrength}
   Alpha=@n5
  )
  n27 = ScalarParameter() {defaultValue=0.5, group=GMap_Glow, parameterName=MechEmissiveVariationNoiseStrength}
  n28 = Multiply(
   A=@n21
   B=Clamp(
    None=ChannelMaskParameter(None=@n25.XYZ Texture) {group=GMap_Glow, parameterName=NoiseChannel}.R
    Min=LinearInterpolate(A=@n26, B=@n27, Alpha=@n4)
   )
  )
  n29 = ScalarParameter() {defaultValue=5, group=GMap_Glow, parameterName=GlowStrength}
  n30 = SetMaterialAttributes!(MaterialAttributes=@n10.MaterialAttributes, EmissiveColor=Add(A=@n10.EmissiveColor, B=Multiply(A=@n28, B=@n29))) {attributeSetTypes [#1]}
  n31 = ChannelMaskParameter(None=FunctionInput() {bUsePreviewValueAsDefault=true, inputName=Mask}) {group=GMap_Glow, parameterName=MaskChannel}
  FunctionOutput(None=StaticSwitchParameter(True=BlendMaterialAttributes(A=@n10.MaterialAttributes, B=@n30, Alpha=@n31.R), False=@n10.MaterialAttributes) {group=Customization, parameterName=UseGmapGlowAndDistort})
 ]}
