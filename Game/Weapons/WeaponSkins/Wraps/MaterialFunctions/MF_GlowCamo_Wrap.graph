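# Material function for the GlowCamo weapon-skin wrap, blending projected camo over incoming material attributes by customization mask, behind the Wrap_GlowCamo static switch.
# Bounds-relative pre-skinned position, with per-weapon, vehicle or mech transforms and panning, feeds top/side projections for color/roughness, normal and emissive mask.
# A timer from a Blueprint-set game time fades out UV warp and masked glow after firing.
# Mobile keeps the incoming normal.
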MaterialFunction /Game/Weapons/WeaponSkins/Wraps/MaterialFunctions/MF_GlowCamo_Wrap
graph {nodes [
  n1 = FunctionInput() {inputName=Material, inputType=FunctionInput_MaterialAttributes}
  n2 = GetMaterialAttributes(None=@n1) {attributeGetTypes [#1, #2, #3]}
  n3 = FunctionInput(Preview=Constant()) {bUsePreviewValueAsDefault=true, inputName=isMech, inputType=FunctionInput_Scalar, sortPriority=4}
  n4 = LinearInterpolate(
   A=ScalarParameter() {defaultValue=1, group=GlowCamo, parameterName=GC_Blend_Contrast, sortPriority=-1}
   B=ScalarParameter() {defaultValue=2, group=GlowCamo, parameterName=GC_Blend_Contrast_Mech, sortPriority=-1}
   Alpha=@n3
  )
  n5 = ScalarParameter() {group=GlowCamo, parameterName=GC_Fill_alternate_Axis_Opacity 01, sortPriority=2}
  n6 = VectorParameter() {group=GlowCamo, parameterName="GC_Fill_alternate_Axis_Color_A=Roughness 01", sortPriority=3}
  n7 = VectorParameter() {group=GlowCamo_UVS, parameterName=GC_Rotation-Weapon-01---------------------, sortPriority=1, defaultValue=<rgba 0 0 -0.25 1>}
  n8 = VectorParameter() {group=GlowCamo_UVS, parameterName=GC_Rotation_Vehicle_01, sortPriority=2, defaultValue=<rgba 0 0 0 1>}
  n9 = FunctionInput(Preview=Constant()) {inputName=isVehicle, inputType=FunctionInput_Scalar, sortPriority=5}
  n10 = VectorParameter() {group=GlowCamo_UVS, parameterName=GC_Rotation_Mech_01, sortPriority=3, defaultValue=<rgba 0.25 0.25 0 1>}
  n11 = LinearInterpolate(A=LinearInterpolate(A=@n7.RGB, B=@n8.RGB, Alpha=@n9), B=@n10.RGB, Alpha=@n3)
  n12 = FunctionInput(Preview=VertexInterpolator(VS=PreSkinnedNormal()).PS) {bUsePreviewValueAsDefault=true, inputName=Pre-Skinned Local Normal, sortPriority=10}
  n13 = VectorParameter() {group=GlowCamo_UVS, parameterName=GC_Offset_Weapon_01---------------------, sortPriority=10, defaultValue=<rgba 0 0 0 1>}
  n14 = VectorParameter() {group=GlowCamo_UVS, parameterName=GC_Offset_Vehicle_01, sortPriority=11, defaultValue=<rgba 0 0 0 1>}
  n15 = VertexInterpolator(VS=PreSkinnedPosition())
  n16 = Add(
   A=MaterialFunctionCall!(
    is Vehicle=@n9
    Object Bounds Max Adjustment=FunctionInput() {bUsePreviewValueAsDefault=true, inputName=Object Bounds Max Adjustment, sortPriority=13}
    Object Bounds Min Adjustment=FunctionInput() {bUsePreviewValueAsDefault=true, inputName=Object Bounds Min Adjustment, sortPriority=12}
    Object Bounds Center Adjustment=FunctionInput() {bUsePreviewValueAsDefault=true, inputName=Object Bounds Center Adjustment, sortPriority=14}
    isMech=@n3
    Side Project Vehicle
    Local Position=FunctionInput(Preview=@n15.PS) {bUsePreviewValueAsDefault=true, inputName=Local Position, sortPriority=11}
   ) {materialFunction=&/MF_WrapBounds}.Result
   B=LinearInterpolate(
    A=LinearInterpolate(A=@n13.RGB, B=@n14.RGB, Alpha=@n9)
    B=VectorParameter() {group=GlowCamo_UVS, parameterName=GC_Offset_Mech_01, sortPriority=12, defaultValue=<rgba 0 0 0 1>}.RGB
    Alpha=@n3
   )
  )
  n17 = VectorParameter() {group=GlowCamo_UVS, parameterName=GC_Panning_Weapon_01---------------------, sortPriority=40, defaultValue=<rgba 0 0 0 1>}
  n18 = VectorParameter() {group=GlowCamo_UVS, parameterName=GC_Panning_Vehicle_01, sortPriority=41, defaultValue=<rgba 0 0 0 1>}
  n19 = VectorParameter() {group=GlowCamo_UVS, parameterName=GC_Panning_Mech_01, sortPriority=42, defaultValue=<rgba 0 0 0 1>}
  n20 = Multiply(A=Time(), B=LinearInterpolate(A=LinearInterpolate(A=@n17.RGB, B=@n18.RGB, Alpha=@n9), B=@n19.RGB, Alpha=@n3))
  n21 = VectorParameter() {group=GlowCamo_UVS, parameterName="GC_Scale_Weapon_01 A=uniform", sortPriority=30, defaultValue=<rgba 1 1 1 1>}
  n22 = VectorParameter() {group=GlowCamo_UVS, parameterName="GC_Scale_Vehicle_01 A=uniform", sortPriority=31, defaultValue=<rgba 1 1 1 1>}
  n23 = LinearInterpolate(A=Multiply(A=@n21.RGB, B=@n21.RGB), B=Multiply(A=@n22.RGB, B=@n22.RGB), Alpha=@n9)
  n24 = VectorParameter() {group=GlowCamo_UVS, parameterName="GC_Scale_Mech_01 A=uniform", defaultValue=<rgba 1 1 1 1>}
  n25 = Multiply(A=@n24.RGB, B=@n24.RGB)
  n26 = TextureCoordinate()
  n27 = LinearInterpolate(
   A=ScalarParameter() {defaultValue=1, group=GlowCamo_UVS, parameterName=GC_WarpUVS-01-TileRate Weapon, sortPriority=53}
   B=ScalarParameter() {defaultValue=1, group=GlowCamo_UVS, parameterName=GC_WarpUVS-01-TileRate Vehicle, sortPriority=54}
   Alpha=@n9
  )
  n28 = Time() {period=1}
  n29 = VectorParameter() {group=GlowCamo_UVS, parameterName=GC_WarpUVS-01-Anitmated Weapon, sortPriority=57, defaultValue=<rgba 0 0 0 1>}
  n30 = VectorParameter() {group=GlowCamo_UVS, parameterName=GC_WarpUVS-01-Anitmated Vehicle, sortPriority=57, defaultValue=<rgba 0 0 0 1>}
  n31 = ComponentMask(None=LinearInterpolate(A=@n29.RGB, B=@n30.RGB, Alpha=@n9)) {g=true, r=true}
  n32 = TextureSampleParameter2D(UVs=Add(A=Multiply(A=@n26, B=@n27), B=Multiply(A=@n28, B=@n31))) {group=GlowCamo_UVS, parameterName=GC_WarpUVS-01-Texture, sortPriority=52, texture=&T_BlacksRGB}
  n33 = LinearInterpolate(
   A=ScalarParameter() {group=GlowCamo_UVS, parameterName=GC_WarpUVS-01-Amount Weapon, sortPriority=55}
   B=ScalarParameter() {group=GlowCamo_UVS, parameterName=GC_WarpUVS-01-Amount_Vehicle, sortPriority=56}
   Alpha=@n9
  )
  n34 = Divide(
   A=Subtract(A=Time(), B=ScalarParameter() {defaultValue=-9999999, group=00-Reactive-GTFB, parameterName=00-GameTimeFromBP, sortPriority=-10})
   B=ScalarParameter() {defaultValue=1, group=GlowCamo_Reactive, parameterName=GC_ReactiveDuration, sortPriority=-9}
  )
  n35 = OneMinus(None=Power(
    Base=Saturate(None=@n34)
    Exp=ScalarParameter() {defaultValue=0.5, group=GlowCamo_Reactive, parameterName=GC_ReactiveFalloff, sortPriority=-8}
   )) {desc=inverts curve from 0-1 to 1-0}
  n36 = Add(
   A=Multiply(A=Add(A=@n16, B=@n20), B=LinearInterpolate(A=@n23, B=@n25, Alpha=@n3))
   B=LinearInterpolate(B=Multiply(A=@n32.R, B=@n33), Alpha=@n35)
  )
  n37 = MaterialFunctionCall!(
   IsCharacter
   IsStatsB
   Texture=TextureObjectParameter() {group=GlowCamo, parameterName=GC_BaseColor_01, sortPriority=11, texture=&T_BlacksRGB}
   Rotation_Axis=@n11
   Caculated_Bounds=@n36
   Pre_Skinned_Normals=@n12
   Blend_Contrast=@n4
   Fill_alternate_Axis_Opacity=@n5
   "Fill_alternate_Axis_Color_A=Roughness"=AppendVector(A=@n6.RGB, B=@n6.RGB)
   Is_Vehicle=@n9
   Is_Mech=@n3
   Enable_Fill_Z_Weapon
   Fill_Z_Weapon_opacity
   Fill_Z_Weapon_Color
  ) {materialFunction=&/MF_TopsSideProjection_Base}
  n38 = Constant3Vector()
  n39 = TextureObjectParameter() {
   group=GlowCamo
   parameterName=GC_Normal_01
   samplerType=SAMPLERTYPE_Normal
   sortPriority=16
   texture=&FlatNormalBlue
  }
  n40 = MaterialFunctionCall!(
   Texture_Normals=@n39
   Rotation_Axis=@n11
   Caculated_Bounds=@n36
   Pre_Skinned_Normals=@n12
   Blend_Contrast=@n4
   Fill_alternate_Axis_Opacity=@n5
   Is_Vehicle=@n9
   Is_Mech=@n3
  ) {materialFunction=&/MF_TopsSideProjection_Normals}
  n41 = LinearInterpolate(
   A=@n2.Normal
   B=Add(
    A=@n2.Normal
    B=StaticSwitchParameter(True=@n38, False=@n40.Add_Surface_Normal) {group=GlowCamo, parameterName=GC_Use_SimpleNormal, sortPriority=15}
   )
   Alpha=ScalarParameter() {defaultValue=1, group=GlowCamo, parameterName=GC_Normal_Intensty, sortPriority=1}
  )
  n42 = VectorParameter() {group=GlowCamo, parameterName=GC_Fill_alternate_Axis_Color_Mask_01, sortPriority=21, defaultValue=<rgba 0 0 0 1>}
  n43 = VectorParameter() {group=GlowCamo, parameterName=GC_Fill_Z_Axis_Weapon_Color_Mask_01, sortPriority=21, defaultValue=<rgba 0 0 0 1>}
  n44 = ChannelMaskParameter(None=MaterialFunctionCall(
    Texture=TextureObjectParameter() {group=GlowCamo, parameterName=GC_Mask_01, sortPriority=20, texture=&T_BlacksRGB}
    Rotation_Axis=@n11
    Caculated_Bounds=@n36
    Pre_Skinned_Normals=@n12
    Blend_Contrast=@n4
    Fill_alternate_Axis_Opacity=@n5
    "Fill_alternate_Axis_Color_A=Roughness"=AppendVector(A=@n42.RGB, B=@n42.RGB)
    Is_Vehicle=@n9
    Is_Mech=@n3
    Enable_Fill_Z_Weapon=StaticBool() {value=true}
    Fill_Z_Weapon_opacity=ScalarParameter() {group=GlowCamo, parameterName=GC_Fill_Z_Axis_Weapon_Opacity 01, sortPriority=2}
    Fill_Z_Weapon_Color=AppendVector(A=@n43.RGB, B=@n43.RGB)
   ) {materialFunction=&MF_TopsSideProjection_Base}.Combined) {group=GlowCamo, maskChannel=Green, parameterName=GC_Mask_EmissiveChannel_01, sortPriority=26, defaultValue=<rgba 0 1 0 0>}
  n45 = Multiply(A=VectorParameter() {group=GlowCamo, parameterName=GC_EmissiveColor, sortPriority=8, defaultValue=<rgba 0 0 0 1>}.RGB, B=@n44.R)
  n46 = ScalarParameter() {defaultValue=1, group=GlowCamo_Reactive, parameterName=GC_Reactive_EmissiveMultiplier}
  n47 = SetMaterialAttributes!(
   MaterialAttributes=@n2.MaterialAttributes
   BaseColor=ComponentMask(None=@n37.Combined) {b=true, g=true, r=true}
   Roughness=LinearInterpolate(
    A=ScalarParameter() {group=GlowCamo, parameterName=GC_Roughness_Min, sortPriority=13}
    B=ScalarParameter() {defaultValue=1, group=GlowCamo, parameterName=GC_Roughness_Max, sortPriority=14}
    Alpha=ComponentMask(None=@n37.Combined) {a=true}
   )
   Normal=ShadingPathSwitch(Default=@n41, Deferred=@n41, Mobile=@n2.Normal) {desc=Migrated from FeatureLevelSwitch}
   Specular=ScalarParameter() {defaultValue=0.5, group=GlowCamo, parameterName=GC_Specular, sortPriority=12}
   Metallic=ScalarParameter() {defaultValue=0.5, group=GlowCamo, parameterName=GC_Metallic, sortPriority=12}
   EmissiveColor=LinearInterpolate(B=Multiply(A=@n45, B=@n46), Alpha=@n35)
  ) {attributeSetTypes [
    #1
    #2
    #3
    #4
    #5
    #6
   ]}
  n48 = Add(
   A=FunctionInput() {bUsePreviewValueAsDefault=true, inputName="Customization Mask (_CM B)", inputType=FunctionInput_Scalar, sortPriority=24}
   B=ScalarParameter() {group=GlowCamo, parameterName=GC_Remove_Customization Mask, sortPriority=-10}
  )
  n49 = StaticSwitchParameter(True=BlendMaterialAttributes(A=@n2.MaterialAttributes, B=@n47, Alpha=Saturate(None=@n48)), False=@n1) {
   desc=Topside Projection Base - warpable - Reactive - simple mask + Color with glow on fire
   group=Customization_Wrap
   parameterName=Wrap_GlowCamo
  }
  FunctionOutput(None=@n49)
 ]}
