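# Material function for the GlowVortex weapon and vehicle wrap, with mech variants, blended over the input material by the customization mask behind a static switch.
# Top/side projection on adjusted bounds, panned and texture-warped, supplies pulsing base color, roughness, normals and masks.
# Emissive combines curvature and inner-sphere glow.
# Event audio amplitude boosts color and glow, while a timer from a Blueprint-set game time adds a decaying front-glow burst on firing.
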MaterialFunction /Game/Weapons/WeaponSkins/Wraps/MaterialFunctions/MF_GlowVortex_Wrap
graph {nodes [
  n1 = FunctionInput(Preview=Constant()) {inputName=isVehicle, inputType=FunctionInput_Scalar, sortPriority=5}
  n2 = FunctionInput() {bUsePreviewValueAsDefault=true, inputName=Object Bounds Max Adjustment, sortPriority=13}
  n3 = FunctionInput() {bUsePreviewValueAsDefault=true, inputName=Object Bounds Min Adjustment, sortPriority=12}
  n4 = FunctionInput() {bUsePreviewValueAsDefault=true, inputName=Object Bounds Center Adjustment, sortPriority=14}
  n5 = FunctionInput(Preview=Constant()) {bUsePreviewValueAsDefault=true, inputName=isMech, inputType=FunctionInput_Scalar, sortPriority=4}
  n6 = FunctionInput(Preview=VertexInterpolator(VS=PreSkinnedPosition()).PS) {bUsePreviewValueAsDefault=true, inputName=Local Position, sortPriority=11}
  n7 = MaterialFunctionCall!(
   is Vehicle=@n1
   Object Bounds Max Adjustment=@n2
   Object Bounds Min Adjustment=@n3
   Object Bounds Center Adjustment=@n4
   isMech=@n5
   Side Project Vehicle
   Local Position=@n6
  ) {materialFunction=&/MF_WrapBounds}
  n8 = MaterialFunctionCall(Color=Multiply(A=@n7.Result, B=Constant3Vector() {constant=<rgba 1 1 -1 0>})) {materialFunction=&SplitComponents}
  AppendVector(A=OneMinus(None=@n8.B), B=@n8.R)
  n9 = TextureCoordinate()
  n10 = LinearInterpolate(
   A=ScalarParameter() {defaultValue=1, group=GlowVortex_UVS, parameterName=GV_WarpUVS-01-TileRate Weapon, sortPriority=53}
   B=ScalarParameter() {defaultValue=1, group=GlowVortex_UVS, parameterName=GV_WarpUVS-01-TileRate Vehicle, sortPriority=54}
   Alpha=@n1
  )
  n11 = Time() {period=1}
  n12 = VectorParameter() {group=GlowVortex_UVS, parameterName=GV_WarpUVS-01-Anitmated Weapon, sortPriority=57, defaultValue=<rgba 0 0 0 1>}
  n13 = VectorParameter() {group=GlowVortex_UVS, parameterName=GV_WarpUVS-01-Anitmated Vehicle, sortPriority=57, defaultValue=<rgba 0 0 0 1>}
  n14 = LinearInterpolate(A=@n12.RGB, B=@n13.RGB, Alpha=@n1)
  n15 = ComponentMask(None=@n14) {g=true, r=true}
  n16 = TextureSampleParameter2D(UVs=Add(A=Multiply(A=@n9, B=@n10), B=Multiply(A=@n11, B=@n15))) {group=GlowVortex_UVS, parameterName=GV_WarpUVS-01-Texture, sortPriority=52, texture=&T_BlacksRGB}
  ChannelMaskParameter(None=@n16.RGB) {group=GlowVortex_UVS, parameterName=GV_WarpUVS-01_Channel, sortPriority=51}
  n17 = CollectionParameter() {collection=&FortniteMaterialParameters, parameterName=EventAudio_Amplitude}
  n18 = Power(Base=Multiply(A=@n17) {constB=0.5})
  n19 = CollectionParameter() {collection=&FortniteMaterialParameters, parameterName=EventAudio_Spectrum}
  n20 = MaterialFunctionCall(Float4=@n19) {materialFunction=&BreakOutFloat4Components}
  n21 = Power(Base=Add(A=Multiply(A=@n20.R) {constB=0.5}, B=Multiply(A=@n20.G) {constB=0.5}))
  Clamp(None=Ceil(None=Add(A=@n18, B=@n21)))
  n22 = FunctionInput() {inputName=Material, inputType=FunctionInput_MaterialAttributes}
  n23 = GetMaterialAttributes(None=@n22) {attributeGetTypes [#1, #2, #3]}
  n24 = VectorParameter() {group=Primary Triplaner DiffusePanningMask, parameterName=GV_Pulse_high, defaultValue=<rgba 0 0.0172 0.0729 1>}
  n25 = VectorParameter() {group=Primary Triplaner DiffusePanningMask, parameterName=GV_Pulse_low, defaultValue=<rgba 2 0.1462 0.038 0.4>}
  n26 = MaterialFunctionCall!(
   Brightness-Low=Constant()
   Brightness-High=Constant() {r=1}
   Frequency=ScalarParameter() {defaultValue=0.65, group=Primary Triplaner DiffusePanningMask, parameterName=pulse_time}
   Sinewave-Offset=Constant() {r=0.5}
   Amplitude
   Clamp Sine Wave?
   Clamp-Min
   Clamp-Max
  ) {materialFunction=&/MF_Pulse_Illuminate}
  n27 = TextureObjectParameter() {group=GlowVortex, parameterName=GV_BaseColor_01, sortPriority=11, texture=&T_BlacksRGB}
  n28 = LinearInterpolate(
   A=ScalarParameter() {defaultValue=1, group=GlowVortex, parameterName=GV_Blend_Contrast, sortPriority=-1}
   B=ScalarParameter() {defaultValue=2, group=GlowVortex, parameterName=GV_Blend_Contrast_Mech, sortPriority=-1}
   Alpha=@n5
  )
  n29 = ScalarParameter() {group=GlowVortex, parameterName=GV_Fill_alternate_Axis_Opacity 01, sortPriority=2}
  n30 = VectorParameter() {group=GlowVortex, parameterName="GV_Fill_alternate_Axis_Color_A=Roughness 01", sortPriority=3}
  n31 = VectorParameter() {group=GlowVortex_UVS, parameterName=GV_Rotation-Weapon-01---------------------, sortPriority=1, defaultValue=<rgba 0 0 -0.25 1>}
  n32 = VectorParameter() {group=GlowVortex_UVS, parameterName=GV_Rotation_Vehicle_01, sortPriority=2, defaultValue=<rgba 0 0 0 1>}
  n33 = VectorParameter() {group=GlowVortex_UVS, parameterName=GV_Rotation_Mech_01, sortPriority=3, defaultValue=<rgba 0.25 0.25 0 1>}
  n34 = LinearInterpolate(A=LinearInterpolate(A=@n31.RGB, B=@n32.RGB, Alpha=@n1), B=@n33.RGB, Alpha=@n5)
  n35 = FunctionInput(Preview=VertexInterpolator(VS=PreSkinnedNormal()).PS) {bUsePreviewValueAsDefault=true, inputName=Pre-Skinned Local Normal, sortPriority=10}
  n36 = VectorParameter() {group=GlowVortex_UVS, parameterName=GV_Offset_Weapon_01---------------------, sortPriority=10, defaultValue=<rgba 0 0 0 1>}
  n37 = VectorParameter() {group=GlowVortex_UVS, parameterName=GV_Offset_Vehicle_01, sortPriority=11, defaultValue=<rgba 0 0 0 1>}
  n38 = VectorParameter() {group=GlowVortex_UVS, parameterName=GV_Offset_Mech_01, sortPriority=12, defaultValue=<rgba 0 0 0 1>}
  n39 = Add(A=@n7.Result, B=LinearInterpolate(A=LinearInterpolate(A=@n36.RGB, B=@n37.RGB, Alpha=@n1), B=@n38.RGB, Alpha=@n5))
  n40 = VectorParameter() {group=GlowVortex_UVS, parameterName=GV_Panning_Weapon_01---------------------, sortPriority=40, defaultValue=<rgba 0 0 0 1>}
  n41 = VectorParameter() {group=GlowVortex_UVS, parameterName=GV_Panning_Vehicle_01, sortPriority=41, defaultValue=<rgba 0 0 0 1>}
  n42 = VectorParameter() {group=GlowVortex_UVS, parameterName=GV_Panning_Mech_01, sortPriority=42, defaultValue=<rgba 0 0 0 1>}
  n43 = Multiply(A=Time(), B=LinearInterpolate(A=LinearInterpolate(A=@n40.RGB, B=@n41.RGB, Alpha=@n1), B=@n42.RGB, Alpha=@n5))
  n44 = VectorParameter() {group=GlowVortex_UVS, parameterName="GV_Scale_Weapon_01 A=uniform", sortPriority=30, defaultValue=<rgba 1 1 1 1>}
  n45 = VectorParameter() {group=GlowVortex_UVS, parameterName="GV_Scale_Vehicle_01 A=uniform", sortPriority=31, defaultValue=<rgba 1 1 1 1>}
  n46 = LinearInterpolate(A=Multiply(A=@n44.RGB, B=@n44.RGB), B=Multiply(A=@n45.RGB, B=@n45.RGB), Alpha=@n1)
  n47 = VectorParameter() {group=GlowVortex_UVS, parameterName="GV_Scale_Mech_01 A=uniform", defaultValue=<rgba 1 1 1 1>}
  n48 = Multiply(A=@n47.RGB, B=@n47.RGB)
  n49 = ScalarParameter() {group=GlowVortex_UVS, parameterName=GV_WarpUVS-01-Amount Weapon, sortPriority=55}
  n50 = ScalarParameter() {group=GlowVortex_UVS, parameterName=GV_WarpUVS-01-Amount_Vehicle, sortPriority=56}
  n51 = Add(
   A=Multiply(A=Add(A=@n39, B=@n43), B=LinearInterpolate(A=@n46, B=@n48, Alpha=@n5))
   B=Multiply(A=@n16.R, B=LinearInterpolate(A=@n49, B=@n50, Alpha=@n1))
  )
  n52 = MaterialFunctionCall!(
   IsCharacter
   IsStatsB
   Texture=@n27
   Rotation_Axis=@n34
   Caculated_Bounds=@n51
   Pre_Skinned_Normals=@n35
   Blend_Contrast=@n28
   Fill_alternate_Axis_Opacity=@n29
   "Fill_alternate_Axis_Color_A=Roughness"=AppendVector(A=@n30.RGB, B=@n30.RGB)
   Is_Vehicle=@n1
   Is_Mech=@n5
   Enable_Fill_Z_Weapon
   Fill_Z_Weapon_opacity
   Fill_Z_Weapon_Color
  ) {materialFunction=&/MF_TopsSideProjection_Base}
  n53 = Multiply(
   A=ComponentMask(None=LinearInterpolate(A=@n24.RGB, B=@n25.RGB, Alpha=@n26.Result)) {b=true, g=true, r=true}
   B=ComponentMask(None=@n52.Combined) {b=true, g=true, r=true}
  )
  n54 = VectorParameter() {group=GlowVortex, parameterName=GV_Fill_alternate_Axis_Color_Mask_01, sortPriority=21, defaultValue=<rgba 0 0 0 1>}
  n55 = VectorParameter() {group=GlowVortex, parameterName=GV_Fill_Z_Axis_Weapon_Color_Mask_01, sortPriority=21, defaultValue=<rgba 0 0 0 1>}
  n56 = MaterialFunctionCall(
   Texture=TextureObjectParameter() {group=GlowVortex, parameterName=GV_Mask_01, sortPriority=20, texture=&T_BlacksRGB}
   Rotation_Axis=@n34
   Caculated_Bounds=@n51
   Pre_Skinned_Normals=@n35
   Blend_Contrast=@n28
   Fill_alternate_Axis_Opacity=@n29
   "Fill_alternate_Axis_Color_A=Roughness"=AppendVector(A=@n54.RGB, B=@n54.RGB)
   Is_Vehicle=@n1
   Is_Mech=@n5
   Enable_Fill_Z_Weapon=StaticBool() {value=true}
   Fill_Z_Weapon_opacity=ScalarParameter() {group=GlowVortex, parameterName=GV_Fill_Z_Axis_Weapon_Opacity 01, sortPriority=2}
   Fill_Z_Weapon_Color=AppendVector(A=@n55.RGB, B=@n55.RGB)
  ) {materialFunction=&MF_TopsSideProjection_Base}
  n57 = ChannelMaskParameter(None=@n56.Combined) {group=GlowVortex_Reactive, maskChannel=Alpha, parameterName=GV_GlowMask_01, sortPriority=26, defaultValue=<rgba 0 0 0 1>}
  n58 = Saturate(None=MaterialFunctionCall(
    Input=@n17
    Input Low=ScalarParameter() {defaultValue=1, group=GlowVortex_Reactive, parameterName=GV_Reactive_AudioLow, sortPriority=9}
    Input High=ScalarParameter() {defaultValue=1, group=GlowVortex_Reactive, parameterName=GV_Reactive_AudioHigh, sortPriority=10}
    Target Low=Constant()
    Target High=Constant() {r=1}
   ) {materialFunction=&RemapValueRange}.Result)
  n59 = ScalarParameter() {defaultValue=1.5, group=GlowVortex_Reactive, parameterName=GV_Reactive_AudioIntensity, sortPriority=5}
  n60 = Multiply(
   A=@n57.R
   B=Power(
    Base=Multiply(A=@n58, B=@n59)
    Exp=ScalarParameter() {defaultValue=1.5, group=GlowVortex_Reactive, parameterName=GV_Reactive_AudioFalloff, sortPriority=6}
   ) {constExponent=1.5}
  )
  n61 = Add(A=Multiply(A=@n53, B=@n60), B=@n53)
  n62 = Multiply(A=Multiply(A=@n53, B=ScalarParameter() {defaultValue=1, group=GlowVortex_Reactive, parameterName=GV_Reactive_BaseColorMultiplier}), B=@n60)
  n63 = ScalarParameter() {group=00-Reactive-GTFB, parameterName=00-GameTimeFromBP, sortPriority=-10}
  n64 = ScalarParameter() {defaultValue=0.15, group=GlowVortex_Reactive, parameterName=GV_ReactiveDelay, sortPriority=-9}
  n65 = Divide(
   A=Subtract(A=Time(), B=Add(A=@n63, B=@n64))
   B=ScalarParameter() {defaultValue=1, group=GlowVortex_Reactive, parameterName=GV_ReactiveDuration, sortPriority=-9}
  )
  n66 = OneMinus(None=Power(
    Base=Saturate(None=@n65)
    Exp=ScalarParameter() {defaultValue=0.5, group=GlowVortex_Reactive, parameterName=GV_ReactiveFalloff, sortPriority=-8}
   )) {desc=inverts curve from 0-1 to 1-0}
  n67 = Constant3Vector()
  n68 = TextureObjectParameter() {
   group=GlowVortex
   parameterName=GV_Normal_01
   samplerType=SAMPLERTYPE_Normal
   sortPriority=16
   texture=&FlatNormalBlue
  }
  n69 = MaterialFunctionCall!(
   Texture_Normals=@n68
   Rotation_Axis=@n34
   Caculated_Bounds=@n51
   Pre_Skinned_Normals=@n35
   Blend_Contrast=@n28
   Fill_alternate_Axis_Opacity=@n29
   Is_Vehicle=@n1
   Is_Mech=@n5
  ) {materialFunction=&/MF_TopsSideProjection_Normals}
  n70 = LinearInterpolate(
   A=@n23.Normal
   B=Add(
    A=@n23.Normal
    B=StaticSwitchParameter(True=@n67, False=@n69.Add_Surface_Normal) {group=GlowVortex, parameterName=GV_Use_SimpleNormal, sortPriority=15}
   )
   Alpha=ScalarParameter() {defaultValue=1, group=GlowVortex, parameterName=GV_Normal_Intensty, sortPriority=1}
  )
  n71 = ChannelMaskParameter(None=@n56.Combined) {group=GlowVortex, maskChannel=Green, parameterName=GV_Mask_01, sortPriority=26, defaultValue=<rgba 0 1 0 0>}
  n72 = GetMaterialAttributes(None=@n22) {attributeGetTypes [#3, #4, #1]}
  n73 = Clamp(
   None=FunctionInput(Preview=TextureSample() {samplerType=SAMPLERTYPE_Masks, texture=/Game/Athena/DrivableVehicles/Golf_Cart/Textures/T_Golf_Cart_Base_CM}.G) {bUsePreviewValueAsDefault=true, inputName="Customization Crevasse (_CM G)", inputType=FunctionInput_Scalar, sortPriority=21}
   Min=ScalarParameter() {defaultValue=0.3, group=Emissive_Curvture, parameterName=Curvature_Clamp_Min, sortPriority=12}
   Max=ScalarParameter() {defaultValue=0.5, group=Emissive_Curvture, parameterName=Curvature_Clamp_Max, sortPriority=13}
  ) {maxDefault=0.5}
  n74 = ScalarParameter() {defaultValue=2.5, group=Emissive_Curvture, parameterName=Curvature_Expo_Weapon, sortPriority=14}
  n75 = ScalarParameter() {defaultValue=2.5, group=Emissive_Curvture, parameterName=Curvature_Expo_Vehicle, sortPriority=14}
  n76 = LinearInterpolate(
   A=Multiply(
    A=Power(Base=@n73, Exp=@n74)
    B=ScalarParameter() {defaultValue=1, group=Emissive_Curvture, parameterName=Curvature_Strength_Weapon, sortPriority=15}
   )
   B=Multiply(
    A=Power(Base=@n73, Exp=@n75)
    B=ScalarParameter() {defaultValue=1, group=Emissive_Curvture, parameterName=Curvature_Strength_Vehicle, sortPriority=15}
   )
   Alpha=@n1
  )
  n77 = LinearInterpolate(
   A=LinearInterpolate(
    A=ScalarParameter() {defaultValue=0.2, group=Emissive_Curvture, parameterName=NormalMap_Weapon_influence, sortPriority=4}
    B=ScalarParameter() {defaultValue=0.1, group=Emissive_Curvture, parameterName=NormalMap_Vehicle_influence, sortPriority=5}
    Alpha=@n1
   )
   B=ScalarParameter() {defaultValue=0.1, group=Emissive_Curvture, parameterName=NormalMap_Mech_influence, sortPriority=5}
   Alpha=@n5
  )
  n78 = Multiply(A=ComponentMask(None=@n72.Normal) {g=true, r=true}, B=@n77)
  n79 = ComponentMask(None=@n35) {b=true, r=true}
  n80 = ComponentMask(None=@n35) {b=true, g=true}
  n81 = Multiply(
   A=LinearInterpolate(A=Abs(None=@n79), B=Abs(None=@n80), Alpha=@n1)
   B=LinearInterpolate(
    A=ScalarParameter() {defaultValue=0.05, group=Emissive_Curvture, parameterName=SkinNormal_influence, sortPriority=6}
    B=ScalarParameter() {defaultValue=0.1, group=Emissive_Curvture, parameterName=SkinNormal_Mech_influence, sortPriority=6}
    Alpha=@n5
   )
  )
  n82 = ObjectPositionWS()
  n83 = Normalize(VectorInput=Subtract(A=@n82, B=CameraPositionWS()))
  n84 = Normalize(VectorInput=CrossProduct(A=@n83, B=Constant3Vector() {constant=<rgba 0 0 1 1>}))
  n85 = VectorParameter() {group=Emissive_Curvture, parameterName=InnerGlow_Weapon_RadialScaleXY, sortPriority=50, defaultValue=<rgba 0.5 0.5 0 1>}
  n86 = VectorParameter() {group=Emissive_Curvture, parameterName=InnerGlow_Vehicle_RadialScaleXY, sortPriority=51, defaultValue=<rgba 0.5 0.5 0 1>}
  n87 = Divide(
   A=ObjectRadius()
   B=LinearInterpolate(A=ComponentMask(None=@n85.RGB) {g=true, r=true}, B=ComponentMask(None=@n86.RGB) {g=true, r=true}, Alpha=@n1)
  )
  n88 = SphereMask(
   A=Add(
    A=Add(A=@n78, B=@n81)
    B=MaterialFunctionCall!(Plane Normal Axis=@n83, Plane X Axis=@n84, Plane Center=@n82, UV Scale=@n87) {materialFunction=/Game/Athena/Environments/Nexus/Materials/VirtualPlaneCoordinatesAbs}.UVs
   )
   B=Constant() {r=0.5}
   Radius=ScalarParameter() {defaultValue=0.5, group=Emissive_Curvture, parameterName=InnerGlow_Radius_A, sortPriority=30}
   Hardness=ScalarParameter() {defaultValue=-2, group=Emissive_Curvture, parameterName=InnerGlow_Hardness_A, sortPriority=31}
  )
  n89 = Add(A=ComponentMask(None=@n72.Normal) {r=true}, B=ComponentMask(None=@n72.Normal) {g=true})
  n90 = PreSkinnedLocalBounds()
  n91 = Add(A=@n90.Min, B=@n3)
  n92 = Subtract(A=Add(A=@n90.Max, B=@n2), B=@n91)
  n93 = Add(A=Subtract(A=@n6, B=@n91), B=Multiply(A=@n92) {constB=-0.5})
  n94 = MaterialFunctionCall(Color=@n92) {materialFunction=&SplitComponents}
  n95 = Max(A=Max(A=@n94.R, B=@n94.R), B=@n94.B)
  n96 = Multiply(A=@n4, B=@n95)
  n97 = ScalarParameter() {defaultValue=0.5, group=Emissive_Curvture, parameterName=InnerGlow_Radius, sortPriority=30}
  n98 = Multiply(
   A=OneMinus(None=Multiply(A=@n89, B=@n77))
   B=SphereMask(
    A=Add(A=@n93, B=@n96)
    B=Constant3Vector() {constant=<rgba 0.5 0.5 0.5 0>}
    Radius=Multiply(A=@n95, B=@n97)
    Hardness=ScalarParameter() {defaultValue=-2, group=Emissive_Curvture, parameterName=InnerGlow_Hardness, sortPriority=31}
   )
  )
  n99 = Add(
   A=Multiply(A=Saturate(None=@n76), B=ScalarParameter() {defaultValue=0.75, group=Emissive_Curvture, parameterName=Curvature_Opactiy, sortPriority=10})
   B=Multiply(
    A=StaticSwitchParameter(True=@n98, False=@n88) {group=Emissive_Curvture, parameterName=InnerGlow_Cheap, sortPriority=23}
    B=ScalarParameter() {defaultValue=0.8, group=Emissive_Curvture, parameterName=InnerGlow Opacity, sortPriority=20}
   )
  ) {constB=0}
  n100 = OneMinus(None=Multiply(A=@n99, B=ScalarParameter() {defaultValue=1, group=Emissive_Curvture, parameterName=Darken_Emissive, sortPriority=-96}))
  n101 = LinearInterpolate(
   A=VectorParameter() {group=Emissive_Curvture, parameterName=Glow_ColorB, sortPriority=8, defaultValue=<rgba 0 0 0 1>}.RGB
   B=VectorParameter() {group=Emissive_Curvture, parameterName=Glow_ColorA, sortPriority=8, defaultValue=<rgba 1 0 0 1>}.RGB
   Alpha=Clamp(None=@n99, Min=ScalarParameter() {defaultValue=0.01, group=Emissive_Curvture, parameterName=Glow_Clamp Min, sortPriority=8}) {maxDefault=0.99, minDefault=0.01}
  )
  n102 = Saturate(None=ChannelMaskParameter(None=@n71.R) {desc=Overrided To TopSideProjection, group=Emissive_Curvture_Mask, parameterName=TS_Mask_01_Channel, sortPriority=144}.R)
  n103 = Add(
   A=Multiply(A=@n72.EmissiveColor, B=Saturate(None=@n100))
   B=Multiply(
    A=ScalarParameter() {defaultValue=1, group=Emissive_Curvture, parameterName=Glow_Emissive_Opacty, sliderMax=1, sortPriority=-95}
    B=Multiply(A=@n101, B=@n102)
   )
  )
  n104 = Add(A=Multiply(A=@n103, B=@n60), B=Multiply(B=@n103) {constA=0.35})
  n105 = OneMinus(None=@n8.B)
  n106 = LinearInterpolate(
   A=@n5
   B=Multiply(A=AppendVector(A=@n105, B=@n8.R)) {constB=0.425}
   Alpha=LinearInterpolate(A=AppendVector(A=@n8.G, B=@n8.G), B=AppendVector(A=@n8.R, B=@n8.R), Alpha=@n1)
  )
  n107 = Power(Base=Multiply(A=MaterialFunctionCall(UV Channel=@n106) {materialFunction=&LinearGradient}.UGradient) {constB=0.8}) {constExponent=3}
  n108 = Multiply(A=@n107)
  n109 = ScalarParameter() {defaultValue=1.5, group=GlowVortex_Reactive, parameterName=GV_Reactive_AudioEmissiveIntensity, sortPriority=15}
  n110 = Multiply(
   A=@n107
   B=Power(
    Base=Multiply(A=@n60, B=@n109)
    Exp=ScalarParameter() {defaultValue=1.5, group=GlowVortex_Reactive, parameterName=GV_Reactive_AudioEmissiveFalloff, sortPriority=16}
   ) {constExponent=1.5}
  )
  n111 = VectorParameter() {group=GlowVortex_Reactive, parameterName=GV_EmissiveFrontGlow, sortPriority=8, defaultValue=<rgba 4 0.6112 0.0128 1>}
  n112 = Multiply(
   A=Multiply(A=Add(A=@n108, B=@n110), B=@n111.RGB)
   B=ScalarParameter() {defaultValue=1, group=GlowVortex_Reactive, parameterName=GV_Reactive_EmissiveMultiplier}
  )
  n113 = SetMaterialAttributes!(
   MaterialAttributes=@n23.MaterialAttributes
   BaseColor=ComponentMask(None=LinearInterpolate(A=@n61, B=@n62, Alpha=@n66)) {b=true, g=true, r=true}
   Roughness=LinearInterpolate(
    A=ScalarParameter() {group=GlowVortex, parameterName=GV_Roughness_Min, sortPriority=13}
    B=ScalarParameter() {defaultValue=1, group=GlowVortex, parameterName=GV_Roughness_Max, sortPriority=14}
    Alpha=ComponentMask(None=@n52.Combined) {a=true}
   )
   Normal=ShadingPathSwitch(Default=@n70, Deferred=@n70, Mobile=@n23.Normal) {desc=Migrated from FeatureLevelSwitch}
   Specular=ScalarParameter() {defaultValue=0.5, group=GlowVortex, parameterName=GV_Specular, sortPriority=12}
   Metallic=Multiply(A=@n71.R, B=ScalarParameter() {defaultValue=0.5, group=GlowVortex, parameterName=GV_Metallic, sortPriority=12})
   EmissiveColor=LinearInterpolate(A=@n104, B=Add(A=@n104, B=@n112), Alpha=@n66)
  ) {attributeSetTypes [
    #1
    #2
    #3
    #5
    #6
    #4
   ]}
  n114 = Add(
   A=FunctionInput() {bUsePreviewValueAsDefault=true, inputName="Customization Mask (_CM B)", inputType=FunctionInput_Scalar, sortPriority=24}
   B=ScalarParameter() {group=GlowVortex, parameterName=GV_Remove_Customization Mask, sortPriority=-10}
  )
  n115 = StaticSwitchParameter(True=BlendMaterialAttributes(A=@n23.MaterialAttributes, B=@n113, Alpha=Saturate(None=@n114)), False=@n22) {
   desc=Topside Projection Base + Curvature Emissive - Reactive w Sound Control -Glow on fire and sound
   group=Customization_Wrap
   parameterName=Wrap_GlowVortex
  }
  FunctionOutput(None=@n115)
  n116 = Multiply(A=@n60) {constB=0.05}
  n117 = Add(A=@n14, B=Multiply(A=@n14, B=Power(Base=@n116) {constExponent=1.25}))
  LinearInterpolate(
   A=@n117
   B=Multiply(A=@n117, B=ScalarParameter() {defaultValue=1.5, group=GlowVortex_Reactive, parameterName=GV_ReactiveWarpingMultiplier})
   Alpha=@n66
  )
 ]}
