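# Honeydew wrap material function for weapons, vehicles, mechs and characters, layering a reactive skin over incoming material attributes behind the Wrap_Honeydew static switch.
# Bounds-based local-position UVs, with per-type scale and offset, sample base, normal and mask textures.
# Time since a Blueprint-supplied timestamp drives a curved fade between two base/normal texture sets on fire.
# Customization masks limit coverage; top/side projection calls remain but do not reach the output.
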MaterialFunction /Game/Weapons/WeaponSkins/Wraps/MaterialFunctions/MF_HoneyDew_Wrap
graph {nodes [
  n1 = Constant() {r=2}
  n2 = Time()
  n3 = Clamp(None=Sine(None=Multiply(A=@n1, B=@n2)) {period=4}, Min=Constant() {r=0.5}, Max=Constant() {r=1})
  Desaturation(None=@n3)
  n4 = VertexInterpolator(VS=PreSkinnedPosition())
  n5 = PreSkinnedLocalBounds()
  FunctionInput(Preview=Divide(A=Subtract(A=@n4.PS, B=@n5.Min), B=Subtract(A=@n5.Max, B=@n5.Max))) {bUsePreviewValueAsDefault=true, inputName=Calculated Bounds, inputType=FunctionInput_Scalar, sortPriority=6}
  n6 = FunctionInput() {
   bUsePreviewValueAsDefault=true
   description=primary
   inputName="Customization Mask 2 (_CM B Top Right)"
   inputType=FunctionInput_Scalar
   sortPriority=25
  }
  n7 = ScalarParameter() {group=Honeydew, parameterName=HD_Remove_Customization Mask, sortPriority=-10}
  n8 = ScalarParameter() {defaultValue=1, group=Honeydew, parameterName=HD_CoverEntireObject, sortPriority=-100}
  n9 = LinearInterpolate(A=Saturate(None=Add(A=@n6, B=@n7)), B=@n8, Alpha=@n8)
  n10 = Saturate(None=Subtract(A=NamedRerouteUsage(), B=NamedRerouteUsage()))
  n11 = FunctionInput() {inputName=Material, inputType=FunctionInput_MaterialAttributes}
  n12 = GetMaterialAttributes(None=@n11) {attributeGetTypes [
    #1
    #2
    #3
    #4
    #5
   ]}
  n13 = VectorParameter() {group=Honeydew_UVS, parameterName="HD_Scale_Weapon_01 A=uniform", sortPriority=30, defaultValue=<rgba 1 1 1 1>}
  n14 = VectorParameter() {group=Honeydew_UVS, parameterName="HD_Scale_Vehicle_01 A=uniform", sortPriority=31, defaultValue=<rgba 1 1 1 1>}
  n15 = Multiply(A=@n14.RGB, B=@n14.RGB)
  n16 = VectorParameter() {
   desc=must activate TS_Vehicle_Side_01 UVs
   group=Honeydew_UVS
   parameterName="HD_Scale_Vehicle_Side_01 A=uniform"
   sortPriority=31
   defaultValue=<rgba 1 1 1 1>
  }
  n17 = Multiply(A=@n16.RGB, B=@n16.RGB)
  n18 = FunctionInput() {bUsePreviewValueAsDefault=true, inputName=Side Project Vehicle, inputType=FunctionInput_Scalar, sortPriority=1}
  n19 = FunctionInput(Preview=Constant()) {inputName=isVehicle, inputType=FunctionInput_Scalar, sortPriority=5}
  n20 = VectorParameter() {group=Honeydew_UVS, parameterName="HD_Scale_Mech_01 A=uniform", defaultValue=<rgba 1 1 1 1>}
  n21 = VectorParameter() {group=Honeydew_UVS, parameterName="HD_Scale_Character_01 A=uniform", defaultValue=<rgba 1 1 1 1>}
  n22 = FunctionInput(Preview=Constant()) {bUsePreviewValueAsDefault=true, inputName=isMech, inputType=FunctionInput_Scalar, sortPriority=4}
  n23 = LinearInterpolate(
   A=LinearInterpolate(A=Multiply(A=@n13.RGB, B=@n13.RGB), B=LinearInterpolate(A=@n15, B=@n17, Alpha=@n18), Alpha=@n19)
   B=LinearInterpolate(A=Multiply(A=@n20.RGB, B=@n20.RGB), B=Multiply(A=@n21.RGB, B=@n21.RGB), Alpha=NamedRerouteUsage())
   Alpha=@n22
  )
  n24 = VertexInterpolator(VS=PreSkinnedPosition())
  n25 = PreSkinnedLocalBounds()
  n26 = FunctionInput(Preview=Divide(A=Subtract(A=@n24.PS, B=@n25.Min), B=Subtract(A=@n25.Max, B=@n25.Max))) {bUsePreviewValueAsDefault=true, inputName=Wrap Bounds, sortPriority=12}
  n27 = PreSkinnedPosition()
  n28 = MaterialFunctionCall!(
   is Vehicle=@n19
   Object Bounds Max Adjustment=FunctionInput() {bUsePreviewValueAsDefault=true, inputName=Object Bounds Max Adjustment, sortPriority=13}
   Object Bounds Min Adjustment=FunctionInput() {bUsePreviewValueAsDefault=true, inputName=Object Bounds Min Adjustment, sortPriority=12}
   Object Bounds Center Adjustment=FunctionInput() {bUsePreviewValueAsDefault=true, inputName=Object Bounds Center Adjustment, sortPriority=14}
   isMech=@n22
   Side Project Vehicle=@n18
   Local Position=FunctionInput(Preview=VertexInterpolator(VS=@n27).PS) {bUsePreviewValueAsDefault=true, inputName=Local Position, sortPriority=11}
  ) {materialFunction=&/MF_WrapBounds}
  n29 = NamedRerouteUsage()
  n30 = VectorParameter() {group=Honeydew_UVS, parameterName=HD_Offset_Weapon_01---------------------, sortPriority=10, defaultValue=<rgba 0 0 0 1>}
  n31 = VectorParameter() {group=Honeydew_UVS, parameterName=HD_Offset_Vehicle_01, sortPriority=11, defaultValue=<rgba 0 0 0 1>}
  n32 = VectorParameter() {group=ConstellationSun, parameterName=CS_TextureOffset_Vehicle_Side, sortPriority=15, defaultValue=<rgba 0 0 0 1>}
  n33 = LinearInterpolate(A=@n31.RGB, B=@n32.RGB, Alpha=@n18)
  n34 = VectorParameter() {group=Honeydew_UVS, parameterName=HD_Offset_Mech_01, sortPriority=12, defaultValue=<rgba 0 0 0 1>}
  n35 = VectorParameter() {group=Honeydew_UVS, parameterName=HD_Offset_Character_01, sortPriority=12, defaultValue=<rgba 0 0 0 1>}
  n36 = NamedRerouteUsage()
  n37 = Add(
   A=Multiply(A=LinearInterpolate(A=@n26, B=@n28.Result, Alpha=@n29), B=Constant3Vector() {constant=<rgba 1 1 -1 0>})
   B=LinearInterpolate(A=LinearInterpolate(A=@n30.RGB, B=@n33, Alpha=@n19), B=LinearInterpolate(A=@n34.RGB, B=@n35.RGB, Alpha=@n36), Alpha=@n22)
  )
  n38 = MaterialFunctionCall(Color=Add(B=Multiply(A=@n37, B=@n23))) {materialFunction=&SplitComponents}
  n39 = AppendVector(A=@n38.R, B=@n38.R)
  n40 = AppendVector(A=@n38.R, B=@n38.R)
  n41 = LinearInterpolate(
   A=LinearInterpolate(A=AppendVector(A=@n38.G, B=@n38.G), B=LinearInterpolate(A=@n39, B=@n40, Alpha=@n18), Alpha=@n19)
   B=AppendVector(A=OneMinus(None=@n38.B), B=@n38.R)
   Alpha=@n22
  )
  n42 = TextureSampleParameter2D(UVs=@n41) {group=Honeydew, parameterName=HD_BaseTexture_01, sortPriority=-3, texture=&T_BlacksRGB}
  n43 = TextureSampleParameter2D(UVs=@n41) {group=Honeydew, parameterName=HD_BaseTexture_02, sortPriority=-3, texture=&T_BlacksRGB}
  n44 = Time()
  n45 = ScalarParameter() {group=00-Reactive-GTFB, parameterName=00-GameTimeFromBP, sortPriority=1}
  n46 = Saturate(None=Divide(
    A=Subtract(A=@n44, B=@n45)
    B=ScalarParameter() {defaultValue=1, group=Honeydew_Reactive, parameterName=HD_AnimationDuration, sortPriority=2}
   ))
  n47 = ScalarParameter() {defaultValue=0.5, group=Honeydew_Reactive, parameterName=HD_Curve_Amount, sortPriority=3}
  n48 = ScalarParameter() {group=Honeydew_Reactive, parameterName=HD_Reactive, sortPriority=-10}
  n49 = LinearInterpolate(A=OneMinus(None=Power(Base=@n46, Exp=@n47)) {desc=inverts curve from 0-1 to 1-0}, B=@n48, Alpha=@n48)
  n50 = LinearInterpolate(A=@n42.RGBA, B=@n43.RGBA, Alpha=@n49) {desc=Base Color}
  n51 = TextureSampleParameter2D(UVs=@n41) {
   group=Honeydew
   parameterName=HD_NormalTexture_01
   samplerType=SAMPLERTYPE_Normal
   sortPriority=-2
   texture=&FlatNormalBlue
  }
  n52 = TextureSampleParameter2D(UVs=@n41) {
   group=Honeydew
   parameterName=HD_NormalTexture_02
   samplerType=SAMPLERTYPE_Normal
   sortPriority=-2
   texture=&FlatNormalBlue
  }
  n53 = ScalarParameter() {defaultValue=1, group=Honeydew_Reactive, parameterName=HD_Normal_IntenstyReactive, sortPriority=1}
  n54 = TextureSampleParameter2D(UVs=@n41) {group=Honeydew, parameterName=HD_MaskTexture_01, sortPriority=-1, texture=&T_BlacksRGB}
  n55 = ComponentMask(None=@n54.RGBA) {r=true}
  n56 = LinearInterpolate(
   A=@n12.Normal
   B=Add(A=@n12.Normal, B=LinearInterpolate(A=@n51.RGB, B=@n52.RGB, Alpha=@n49) {desc=Normal})
   Alpha=LinearInterpolate(
    A=ScalarParameter() {defaultValue=1, group=Honeydew, parameterName=HD_Normal_Intensty, sortPriority=1}
    B=Multiply(A=@n53, B=@n55)
    Alpha=@n49
   )
  )
  n57 = ChannelMaskParameter(None=@n54.RGBA) {group=Honeydew, maskChannel=Alpha, parameterName=HD_SpecChannel_Mask_01, sortPriority=26, defaultValue=<rgba 0 0 0 1>}
  n58 = LinearInterpolate(
   A=ScalarParameter() {group=Honeydew, parameterName=HD_Specularty_Min, sortPriority=17}
   B=ScalarParameter() {defaultValue=1, group=Honeydew, parameterName=HD_Specularty_Max, sortPriority=17}
   Alpha=@n57.R
  )
  n59 = ChannelMaskParameter(None=@n54.RGBA) {group=Honeydew, maskChannel=Alpha, parameterName=HD_MetalChannel_Mask_01, sortPriority=26, defaultValue=<rgba 0 0 0 1>}
  n60 = SetMaterialAttributes!(
   MaterialAttributes=@n12.MaterialAttributes
   BaseColor=ComponentMask(None=@n50) {b=true, g=true, r=true}
   Roughness=LinearInterpolate(
    A=ScalarParameter() {group=Honeydew, parameterName=HD_Roughness_Min, sortPriority=13}
    B=ScalarParameter() {defaultValue=1, group=Honeydew, parameterName=HD_Roughness_Max, sortPriority=14}
    Alpha=ComponentMask(None=@n50) {a=true}
   )
   Normal=ShadingPathSwitch(Default=@n56, Deferred=@n56, Mobile=@n12.Normal) {desc=Migrated from FeatureLevelSwitch}
   Specular=@n58
   Metallic=@n59.R
  ) {attributeSetTypes [
    #1
    #2
    #3
    #6
    #7
   ]}
  n61 = StaticSwitchParameter(True=BlendMaterialAttributes(A=@n12.MaterialAttributes, B=@n60, Alpha=LinearInterpolate(A=@n9, B=@n6, Alpha=@n10)), False=@n11) {desc=Side Projection Base - Reactive - Normal and Base Texture lerp on Fire, group=Customization_Wrap, parameterName=Wrap_Honeydew}
  FunctionOutput(None=@n61)
  n62 = Fresnel(ExponentIn=ScalarParameter() {defaultValue=10, group=Honeydew, parameterName=HD_SpecularFresnelExponent, sortPriority=17})
  n63 = ScalarParameter() {defaultValue=0.5, group=Honeydew, parameterName=HD_SpecularFresnelIntensity, sortPriority=17}
  LinearInterpolate(
   A=@n58
   B=Add(A=@n58, B=Multiply(A=@n62, B=@n63))
   Alpha=ScalarParameter() {group=Honeydew, parameterName=HD_UseSpecularFresnel, sortPriority=16}
  )
  n64 = VectorParameter() {group=Honeydew_UVS, parameterName=HD_Offset_Mech_01, sortPriority=12, defaultValue=<rgba 0 0 0 1>}
  n65 = VectorParameter() {group=Honeydew_UVS, parameterName=HD_Offset_Character_01, sortPriority=12, defaultValue=<rgba 0 0 0 1>}
  n66 = FunctionInput() {bUsePreviewValueAsDefault=true, inputName=IsCharacter, inputType=FunctionInput_Scalar, sortPriority=1}
  LinearInterpolate(A=@n64.RGB, B=@n65.RGB, Alpha=@n66)
  n67 = TextureObjectParameter() {group=Honeydew, parameterName=HD_BaseColor_01, sortPriority=11, texture=/Game/Weapons/WeaponSkins/Wraps/Textures/UVCheaker}
  n68 = LinearInterpolate(
   A=ScalarParameter() {defaultValue=1, group=Honeydew, parameterName=HD_Blend_Contrast, sortPriority=-1}
   B=ScalarParameter() {defaultValue=2, group=Honeydew, parameterName=HD_Blend_Contrast_Mech, sortPriority=-1}
   Alpha=@n22
  )
  n69 = ScalarParameter() {group=Honeydew, parameterName=HD_Fill_alternate_Axis_Opacity 01, sortPriority=2}
  n70 = VectorParameter() {group=Honeydew, parameterName="HD_Fill_alternate_Axis_Color_A=Roughness 01", sortPriority=3}
  n71 = StaticBool() {value=true}
  n72 = FunctionInput(Preview=VertexInterpolator(VS=PreSkinnedNormal()).PS) {bUsePreviewValueAsDefault=true, inputName=Pre-Skinned Local Normal, sortPriority=10}
  n73 = VectorParameter() {group=Honeydew_UVS, parameterName=HD-Rotation-Weapon-01---------------------, sortPriority=1, defaultValue=<rgba 0 0 -0.25 1>}
  n74 = VectorParameter() {group=Honeydew_UVS, parameterName=HD_Rotation_Vehicle_01, sortPriority=2, defaultValue=<rgba 0 0 0 1>}
  n75 = VectorParameter() {
   desc=must activate TS_Vehicle_Side_01 UVs
   group=Honeydew_UVS
   parameterName=HD_Rotation_Vehicle_Side_01
   sortPriority=2
   defaultValue=<rgba 0 0 0 1>
  }
  n76 = Multiply(A=@n18, B=ScalarParameter() {group=Honeydew_UVS, parameterName="HD_Vehicle_Side_01 UVs 0=Off 1=Active"})
  n77 = VectorParameter() {group=Honeydew_UVS, parameterName=HD_Rotation_Mech_01, sortPriority=3, defaultValue=<rgba 0.25 0.25 0 1>}
  n78 = VectorParameter() {group=Honeydew_UVS, parameterName=HD_Rotation_Character_01, sortPriority=3, defaultValue=<rgba 0.25 0.25 0 1>}
  n79 = LinearInterpolate(
   A=LinearInterpolate(A=@n73.RGB, B=LinearInterpolate(A=@n74.RGB, B=@n75.RGB, Alpha=@n76), Alpha=@n19)
   B=LinearInterpolate(A=@n77.RGB, B=@n78.RGB, Alpha=@n66)
   Alpha=@n22
  )
  n80 = VectorParameter() {group=Honeydew_UVS, parameterName=HD_Offset_Weapon_01---------------------, sortPriority=10, defaultValue=<rgba 0 0 0 1>}
  n81 = VectorParameter() {group=Honeydew_UVS, parameterName=HD_Offset_Vehicle_01, sortPriority=11, defaultValue=<rgba 0 0 0 1>}
  n82 = VectorParameter() {
   desc=must activate TS_Vehicle_Side_01 UVs
   group=Honeydew_UVS
   parameterName=HD_Offset_Vehicle_Side_01
   sortPriority=11
   defaultValue=<rgba 0 0 0 1>
  }
  n83 = LinearInterpolate(A=@n80.RGB, B=LinearInterpolate(A=@n81.RGB, B=@n82.RGB, Alpha=@n76), Alpha=@n19)
  n84 = VectorParameter() {group=Honeydew_UVS, parameterName="HD_Scale_Weapon_01 A=uniform", sortPriority=30, defaultValue=<rgba 1 1 1 1>}
  n85 = Multiply(A=@n84.RGB, B=@n84.RGB)
  n86 = VectorParameter() {group=Honeydew_UVS, parameterName="HD_Scale_Vehicle_01 A=uniform", sortPriority=31, defaultValue=<rgba 1 1 1 1>}
  n87 = VectorParameter() {
   desc=must activate TS_Vehicle_Side_01 UVs
   group=Honeydew_UVS
   parameterName="HD_Scale_Vehicle_Side_01 A=uniform"
   sortPriority=31
   defaultValue=<rgba 1 1 1 1>
  }
  n88 = LinearInterpolate(A=Multiply(A=@n86.RGB, B=@n86.RGB), B=Multiply(A=@n87.RGB, B=@n87.RGB), Alpha=@n76)
  n89 = VectorParameter() {group=Honeydew_UVS, parameterName="HD_Scale_Mech_01 A=uniform", defaultValue=<rgba 1 1 1 1>}
  n90 = Multiply(A=@n89.RGB, B=@n89.RGB)
  n91 = VectorParameter() {group=Honeydew_UVS, parameterName="HD_Scale_Character_01 A=uniform", defaultValue=<rgba 1 1 1 1>}
  n92 = Multiply(A=@n91.RGB, B=@n91.RGB)
  n93 = Multiply(
   A=Add(A=@n28.Result, B=LinearInterpolate(A=@n83, B=@n64.RGB, Alpha=@n22))
   B=LinearInterpolate(A=LinearInterpolate(A=@n85, B=@n88, Alpha=@n19), B=LinearInterpolate(A=@n90, B=@n92, Alpha=@n66), Alpha=@n22)
  )
  MaterialFunctionCall!(
   IsCharacter
   IsStatsB
   Texture=@n67
   Rotation_Axis=@n79
   Caculated_Bounds=@n93
   Pre_Skinned_Normals=@n72
   Blend_Contrast=@n68
   Fill_alternate_Axis_Opacity=@n69
   "Fill_alternate_Axis_Color_A=Roughness"=AppendVector(A=@n70.RGB, B=@n70.RGB)
   Is_Vehicle=@n19
   Is_Mech=@n22
   Enable_Fill_Z_Weapon=@n71
   Fill_Z_Weapon_opacity
   Fill_Z_Weapon_Color
  ) {materialFunction=&/MF_TopsSideProjection_Base}
  n94 = TextureObjectParameter() {
   group=Honeydew
   parameterName=HD_Normal_01
   samplerType=SAMPLERTYPE_Normal
   sortPriority=16
   texture=&FlatNormalBlue
  }
  MaterialFunctionCall!(
   Texture_Normals=@n94
   Rotation_Axis=@n79
   Caculated_Bounds=@n93
   Pre_Skinned_Normals=@n72
   Blend_Contrast=@n68
   Fill_alternate_Axis_Opacity=@n69
   Is_Vehicle=@n19
   Is_Mech=@n22
  ) {materialFunction=&/MF_TopsSideProjection_Normals}
  n95 = TextureObjectParameter() {group=Honeydew, parameterName=HD_BaseColor_02, sortPriority=111, texture=&UVCheaker}
  n96 = VectorParameter() {group=Honeydew, parameterName=HD_Fill_alternate_Axis_Color_Mask_01, sortPriority=21, defaultValue=<rgba 0 0 0 1>}
  n97 = AppendVector(A=@n96.RGB, B=@n96.RGB)
  MaterialFunctionCall(
   Texture=@n95
   Rotation_Axis=@n79
   Caculated_Bounds=@n93
   Pre_Skinned_Normals=@n72
   Blend_Contrast=@n68
   Fill_alternate_Axis_Opacity=@n69
   "Fill_alternate_Axis_Color_A=Roughness"=@n97
   Is_Vehicle=@n19
   Is_Mech=@n22
  ) {materialFunction=&MF_TopsSideProjection_Base}
  n98 = TextureObjectParameter() {
   group=Honeydew
   parameterName=HD_Normal_02
   samplerType=SAMPLERTYPE_Normal
   sortPriority=116
   texture=&FlatNormalBlue
  }
  MaterialFunctionCall(
   Texture_Normals=@n98
   Rotation_Axis=@n79
   Caculated_Bounds=@n93
   Pre_Skinned_Normals=@n72
   Blend_Contrast=@n68
   Fill_alternate_Axis_Opacity=@n69
   Is_Vehicle=@n19
   Is_Mech=@n22
  ) {materialFunction=&MF_TopsSideProjection_Normals}
  MaterialFunctionCall(
   Texture=TextureObjectParameter() {group=Honeydew, parameterName=HD_Mask_01, sortPriority=20, texture=&T_BlacksRGB}
   Rotation_Axis=@n79
   Caculated_Bounds=@n93
   Pre_Skinned_Normals=@n72
   Blend_Contrast=@n68
   Fill_alternate_Axis_Opacity=@n69
   "Fill_alternate_Axis_Color_A=Roughness"=@n97
   Is_Vehicle=@n19
   Is_Mech=@n22
   Enable_Fill_Z_Weapon=@n71
  ) {materialFunction=&MF_TopsSideProjection_Base}
  Multiply(A=ScalarParameter() {defaultValue=4, group=Honeydew_Reactive, parameterName=HD_Normal_Intensity_Reactive, sortPriority=1}, B=@n3)
  NamedRerouteDeclaration(None=@n66) {name=IsCharacter, nodeColor=<rgba 1 0.9439 0 1>}
  NamedRerouteDeclaration(None=FunctionInput() {bUsePreviewValueAsDefault=true, inputName=isCharAccent, inputType=FunctionInput_Scalar, sortPriority=78}) {name=isCharAccent, nodeColor=<rgba 0 0.9253 1 1>}
  Reroute(None=@n57.R)
 ]}
