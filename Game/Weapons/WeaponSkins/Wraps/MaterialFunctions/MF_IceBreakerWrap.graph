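# IceBreaker wrap material function that layers cracked ice over the incoming material, behind the UseIceBreakerWrap switch, with weapon, vehicle and mech tuning.
# Bounds gradients and Fresnel drive curve-atlas colors, with parallax cracks, interior glow and faked facet reflections (non-mobile).
# A masked undercoat blends to full ice through a sphere mask eroded by crack noise.
# Time since IB_InputGameTime shrinks that ice layer over IB_RefreezeDuration.
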MaterialFunction /Game/Weapons/WeaponSkins/Wraps/MaterialFunctions/MF_IceBreakerWrap
graph {nodes [
  n1 = ScalarParameter() {defaultValue=10, group=IceBreaker, parameterName=IB_CrackBrightness}
  n2 = VertexInterpolator(VS=PreSkinnedPosition())
  n3 = Constant()
  n4 = FunctionInput(Preview=@n3) {bUsePreviewValueAsDefault=true, inputName=isVehicle, inputType=FunctionInput_Scalar}
  n5 = LinearInterpolate(
   A=VectorParameter() {group=IceBreaker, parameterName=IB_Weapon_GradientCenterOffset, defaultValue=<rgba 0 0 0 1>}.RGB
   B=VectorParameter() {group=IceBreaker, parameterName=IB_Vehicle_GradientCenterOffset, defaultValue=<rgba 0 0 0 1>}.RGB
   Alpha=@n4
  )
  n6 = VectorParameter() {group=IceBreaker, parameterName=IB_Mech_GradientCenterOffset, defaultValue=<rgba 0 0 0 1>}
  n7 = FunctionInput(Preview=@n3) {bUsePreviewValueAsDefault=true, inputName=isMech, inputType=FunctionInput_Scalar}
  n8 = PreSkinnedLocalBounds()
  n9 = Divide(A=Add(A=@n2.PS, B=LinearInterpolate(A=@n5, B=@n6.RGB, Alpha=@n7)), B=@n8.Max)
  n10 = ChannelMaskParameter(None=@n9) {group=IceBreaker, maskChannel=Green, parameterName=IB_WeaponGradientDirection, defaultValue=<rgba 0 1 0 0>}
  n11 = ChannelMaskParameter(None=@n9) {group=IceBreaker, maskChannel=Green, parameterName=IB_VehicleGradientDirection, defaultValue=<rgba 0 1 0 0>}
  n12 = ChannelMaskParameter(None=@n9) {group=IceBreaker, maskChannel=Blue, parameterName=IB_MechGradientDirection, defaultValue=<rgba 0 0 1 0>}
  n13 = Clamp(None=LinearInterpolate(A=LinearInterpolate(A=@n10.R, B=@n11.R, Alpha=@n4), B=@n12.R, Alpha=@n7))
  n14 = TextureCoordinate()
  n15 = VectorParameter() {group=IceBreaker, parameterName=IB_Vehicle_IceNormal_02_Tiling, defaultValue=<rgba 2 2 0 1>}
  n16 = VectorParameter() {group=IceBreaker, parameterName=IB_Mech_IceNormal_02_Tiling, defaultValue=<rgba 2 2 0 1>}
  n17 = ComponentMask(None=LinearInterpolate(
    A=VectorParameter() {group=IceBreaker, parameterName=IB_Weapon_IceNormal_02_Tiling, defaultValue=<rgba 2 2 0 1>}.RGB
    B=LinearInterpolate(A=@n15.RGB, B=@n16.RGB, Alpha=@n7)
    Alpha=@n4
   )) {g=true, r=true}
  n18 = TextureSampleParameter2D(UVs=BumpOffset(
    Coordinate=Multiply(A=@n14, B=@n17)
    Height=ScalarParameter() {defaultValue=-0.1, group=IceBreaker, parameterName=IB_HeightOffset}
    HeightRatioInput=ScalarParameter() {defaultValue=0.1, group=IceBreaker, parameterName=IB_HeightRatio}
   )) {
   group=IceBreaker
   parameterName=IceNoise
   samplerType=SAMPLERTYPE_Normal
   texture=/Game/Weapons/WeaponSkins/Wraps/Textures/IceBreaker/T_IceNormal
  }
  n19 = MaterialFunctionCall(Normal=@n18.RGB, Flatness=ScalarParameter() {defaultValue=0.75, group=IceBreaker, parameterName=IB_IceNormalFlatness}) {materialFunction=&FlattenNormal}
  n20 = GetMaterialAttributes(None=FunctionInput(Preview=MakeMaterialAttributes!(
     BaseColor
     Metallic
     Specular
     Roughness
     Anisotropy
     EmissiveColor
     Opacity
     OpacityMask
     Normal
     Tangent
     WorldPositionOffset
     SubsurfaceColor
     ClearCoat
     ClearCoatRoughness
     AmbientOcclusion
     Refraction
     CustomizedUVs_0
     CustomizedUVs_1
     CustomizedUVs_2
     CustomizedUVs_3
     CustomizedUVs_4
     CustomizedUVs_5
     CustomizedUVs_6
     CustomizedUVs_7
     PixelDepthOffset
     ShadingModel
     Displacement
    )) {bUsePreviewValueAsDefault=true, inputName=Material, inputType=FunctionInput_MaterialAttributes}) {attributeGetTypes [#1, #2]}
  n21 = Fresnel(
   ExponentIn=ScalarParameter() {defaultValue=2, group=IceBreaker, parameterName=IB_FresnelExponent}
   Normal=Transform(None=MaterialFunctionCall(BaseNormal=@n20.Normal, AdditionalNormal=@n19.Result) {materialFunction=&BlendAngleCorrectedNormals}.Result)
  )
  n22 = ScalarParameter() {defaultValue=2, group=IceBreaker, parameterName=IB_FresnelPower}
  n23 = DotProduct(A=CameraVectorWS(), B=VertexNormalWS())
  n24 = Multiply(A=Clamp(None=ComponentMask(None=@n18.RGB) {r=true}), B=Clamp(None=Power(Base=@n23) {constExponent=12}))
  n25 = Multiply(A=@n13, B=Max(A=Power(Base=@n21, Exp=@n22), B=@n24))
  n26 = CurveAtlasRowParameter(CurveTime=Clamp(None=Add(A=@n13, B=@n25))) {
   atlas=&AthenaCurveAtlas01
   curve=/Game/Weapons/WeaponSkins/Wraps/Textures/IceBreaker/IceBreakerColorCurve
   defaultValue=199
   group=IceBreaker
   parameterName=IceBreakerBaseColorCurve
  }
  n27 = Multiply(
   A=VectorParameter() {group=IceBreaker, parameterName=IB_EdgeColor, defaultValue=<rgba 1 1 1 1>}.RGB
   B=ScalarParameter() {defaultValue=5, group=IceBreaker, parameterName=IB_EdgeBrightness}
  )
  n28 = ScalarParameter() {defaultValue=2, group=IceBreaker, parameterName=IB_BaseFresnelExponent}
  n29 = Transform(None=@n20.Normal)
  n30 = Clamp(None=Power(Base=Fresnel(ExponentIn=@n28, Normal=@n29), Exp=ScalarParameter() {defaultValue=2, group=IceBreaker, parameterName=IB_BaseFresnelPower}))
  n31 = LinearInterpolate(A=@n26, B=@n27, Alpha=@n30)
  n32 = Subtract(A=@n2.PS, B=@n8.Min)
  n33 = Subtract(A=@n8.Max, B=@n8.Max)
  n34 = ScalarParameter() {defaultValue=1, group=IceBreaker, parameterName=IB_ShatterRadius Multiplier}
  n35 = SphereMask(
   A=Add(A=Divide(A=@n32, B=@n33), B=Constant3Vector() {constant=<rgba -0.5 -0.5 -0.5 0>})
   B=VectorParameter() {group=IceBreaker, parameterName=TransitionPosition_Offset, defaultValue=<rgba 0 0 0 1>}.RGB
   Radius=@n34
   Hardness=ScalarParameter() {group=IceBreaker, parameterName=IB_ShatterGradientHardness}
  )
  n36 = OneMinus(None=@n35)
  n37 = ScalarParameter() {group=IceBreaker, parameterName=IB_CrackGrowth}
  n38 = OneMinus(None=@n36)
  n39 = TextureCoordinate()
  n40 = VectorParameter() {group=IceBreaker, parameterName=IB_Vehicle_Cracks_Tiling, defaultValue=<rgba 2 2 0 1>}
  n41 = VectorParameter() {group=IceBreaker, parameterName=IB_Mech_Cracks_Tiling, defaultValue=<rgba 2 2 0 1>}
  n42 = ComponentMask(None=LinearInterpolate(
    A=VectorParameter() {group=IceBreaker, parameterName=IB_Weapon_Cracks_Tiling, defaultValue=<rgba 2 2 0 1>}.RGB
    B=LinearInterpolate(A=@n40.RGB, B=@n41.RGB, Alpha=@n7)
    Alpha=@n4
   )) {g=true, r=true}
  n43 = TextureSampleParameter2D(UVs=BumpOffset(
    Coordinate=Multiply(A=@n39, B=@n42)
    Height=ScalarParameter() {defaultValue=-0.1, group=IceBreaker, parameterName=IB_HeightOffset}
    HeightRatioInput=ScalarParameter() {defaultValue=0.1, group=IceBreaker, parameterName=IB_HeightRatio}
   )) {
   group=IceBreaker
   parameterName=IB_ErosionNoise
   samplerType=SAMPLERTYPE_Masks
   texture=/Game/Weapons/WeaponSkins/Wraps/Textures/IceBreaker/T_IceBreakerCracks
  }
  n44 = ComponentMask(None=@n43.RGB) {b=true}
  n45 = Subtract(
   A=Subtract(A=@n36, B=OneMinus(None=@n37))
   B=Multiply(A=Multiply(A=@n38, B=@n44), B=ScalarParameter() {defaultValue=1, group=IceBreaker, parameterName=IB_CracksNoiseSpread})
  )
  n46 = LinearInterpolate(A=Constant() {r=1}, B=Ceil(None=Clamp(None=@n45)), Alpha=@n44)
  n47 = CameraVectorWS()
  n48 = CurveAtlasRowParameter(CurveTime=Clamp(None=DotProduct(A=@n47, B=@n29))) {
   atlas=&AthenaCurveAtlas01
   curve=/Game/Weapons/WeaponSkins/Wraps/Textures/IceBreaker/IceReflectanceCurve
   defaultValue=198
   group=IceBreaker
   parameterName=IB_IceReflectanceCurve
  }
  n49 = Multiply(A=LinearInterpolate(A=Multiply(A=@n31, B=@n1), B=@n31, Alpha=@n46), B=@n48.R)
  n50 = ScalarParameter() {defaultValue=0.25, group=IceBreaker, parameterName=IB_UnderCoatBLendValue}
  n51 = LinearInterpolate(
   A=VectorParameter() {group=IceBreaker, parameterName=IB_Weapon_IceNormal_Tiling, defaultValue=<rgba 2 2 0 1>}.RGB
   B=LinearInterpolate(
    A=VectorParameter() {group=IceBreaker, parameterName=IB_Vehicle_IceNormal_Tiling, defaultValue=<rgba 2 2 0 1>}.RGB
    B=VectorParameter() {group=IceBreaker, parameterName=IB_Mech_IceNormal_Tiling, defaultValue=<rgba 2 2 0 1>}.RGB
    Alpha=@n7
   )
   Alpha=@n4
  )
  n52 = TextureSampleParameter2D(UVs=Multiply(A=TextureCoordinate(), B=ComponentMask(None=@n51) {g=true, r=true})) {
   group=IceBreaker
   parameterName=IB_SurfaceNormal
   samplerType=SAMPLERTYPE_Normal
   texture=/Game/Global/Textures/Default/Noise/T_Normal_Noise01
  }
  n53 = MaterialFunctionCall(BaseNormal=@n46, AdditionalNormal=@n52.RGB) {materialFunction=&BlendAngleCorrectedNormals}
  n54 = Add(
   A=Multiply(A=Multiply(A=@n26, B=@n24), B=ScalarParameter() {defaultValue=2, group=IceBreaker, parameterName=IB_InteriorGLow})
   B=Multiply(A=Multiply(A=@n27, B=@n30), B=ScalarParameter() {defaultValue=1, group=IceBreaker, parameterName=IB_EdgeEmissiveStrength})
  )
  n55 = LinearInterpolate(
   A=VectorParameter() {group=IceBreaker, parameterName=IB_Weapon_Shatter_Tiling, defaultValue=<rgba 2 2 0 1>}.RGB
   B=LinearInterpolate(
    A=VectorParameter() {group=IceBreaker, parameterName=IB_Vehicle_Shatter_Tiling, defaultValue=<rgba 2 2 0 1>}.RGB
    B=VectorParameter() {group=IceBreaker, parameterName=IB_Mech_Shatter_Tiling, defaultValue=<rgba 2 2 0 1>}.RGB
    Alpha=@n7
   )
   Alpha=@n4
  )
  n56 = BumpOffset(
   Coordinate=Multiply(A=TextureCoordinate(), B=ComponentMask(None=@n55) {g=true, r=true})
   Height=ScalarParameter() {defaultValue=-0.1, group=IceBreaker, parameterName=IB_HeightOffset}
   HeightRatioInput=ScalarParameter() {defaultValue=0.1, group=IceBreaker, parameterName=IB_HeightRatio}
  )
  n57 = TextureSampleParameter2D(UVs=@n56) {
   group=IceBreaker
   parameterName=IB_Facet_Normal
   samplerType=SAMPLERTYPE_Normal
   texture=/Game/Weapons/WeaponSkins/Wraps/Textures/IceBreaker/T_IceFacet_Normal
  }
  n58 = ComponentMask(None=ReflectionVectorWS(CustomWorldNormal=Transform(None=@n57.RGB) {transformSourceType=TRANSFORMSOURCE_World, transformType=TRANSFORM_Tangent})) {g=true, r=true}
  n59 = Multiply(
   A=TextureSampleParameter2D(UVs=@n58) {parameterName=" IB_FakeReflectionNoise", texture=&T_Inky_Smoke_Tile_Chromatic}.RGB
   B=VectorParameter() {group=IceBreaker, parameterName=Reflection_Tint, defaultValue=<rgba 0.1 0.1 0.1 1>}.RGB
  )
  n60 = Add(
   A=Multiply(A=@n49, B=ScalarParameter() {defaultValue=0.2, group=IceBreaker, parameterName=IB_ColorEmissiveStrength})
   B=ShadingPathSwitch(Default=Add(A=@n54, B=@n59), Mobile=@n54)
  )
  n61 = SetMaterialAttributes!(
   MaterialAttributes=@n20.MaterialAttributes
   BaseColor=LinearInterpolate(A=VectorParameter() {parameterName=IB_UndercoatBaseColor, defaultValue=<rgba 0.026 0.026 0.026 1>}.RGB, B=@n49, Alpha=@n50)
   Metallic=ScalarParameter() {defaultValue=1, parameterName=IB_UnderCoat_Metallic}
   Specular=ScalarParameter() {defaultValue=1, parameterName=IB_UnderCoat_Specular}
   Roughness=ScalarParameter() {defaultValue=0.1, parameterName=IB_UnderCoat_Roughness}
   Normal=ShadingPathSwitch(Default=LinearInterpolate(A=@n20.Normal, B=@n53.Result, Alpha=@n50), Mobile=@n20.Normal)
   EmissiveColor=Multiply(A=@n60, B=@n50) {constB=0.25}
  ) {attributeSetTypes [
    #2
    #3
    #4
    #5
    #1
    #6
   ]}
  n62 = FunctionInput(Preview=Constant() {r=1}) {bUsePreviewValueAsDefault=true, inputName=Mask, inputType=FunctionInput_Scalar}
  n63 = BlendMaterialAttributes(A=@n20.MaterialAttributes, B=@n61, Alpha=@n62) {vertexAttributeBlendType=UseA}
  n64 = SetMaterialAttributes!(
   MaterialAttributes=@n20.MaterialAttributes
   BaseColor=@n49
   Metallic=ScalarParameter() {group=IceBreaker, parameterName=IB_IceMetallic}
   Specular=ScalarParameter() {defaultValue=1, group=IceBreaker, parameterName=IB_IceSpecular}
   Roughness=ScalarParameter() {defaultValue=0.1, group=IceBreaker, parameterName=IB_IceRoughness}
   EmissiveColor=@n60
   Normal=ShadingPathSwitch(Default=@n53.Result, Mobile=@n20.Normal)
  ) {attributeSetTypes [
    #2
    #3
    #4
    #5
    #6
    #1
   ]}
  n65 = ScalarParameter() {defaultValue=2, group=IceBreaker, parameterName=IB_BaseGrowth}
  n66 = Subtract(A=Time(), B=ScalarParameter() {defaultValue=-99999998430674944, parameterName=IB_InputGameTime})
  n67 = ScalarParameter() {defaultValue=2, parameterName=IB_RefreezeDuration}
  n68 = OneMinus(None=Saturate(None=Divide(A=@n66, B=@n67)))
  n69 = Subtract(A=@n36, B=OneMinus(None=Multiply(A=@n65, B=@n68)))
  n70 = TextureSampleParameter2D(UVs=@n56) {
   group=IceBreaker
   parameterName=ErosionNoise
   samplerType=SAMPLERTYPE_Masks
   texture=&T_IceBreakerCracks
  }
  n71 = Multiply(
   A=ChannelMaskParameter(None=@n70.RGB) {group=IceBreaker, maskChannel=Green, parameterName=TearNoisechannel, defaultValue=<rgba 0 1 0 0>}.R
   B=ScalarParameter() {defaultValue=1, group=IceBreaker, parameterName=IB_NoiseSpread}
  )
  n72 = Multiply(A=Clamp(None=Subtract(A=@n69, B=@n71)), B=@n62)
  FunctionOutput(None=StaticSwitchParameter(True=BlendMaterialAttributes(A=@n63, B=@n64, Alpha=@n72) {vertexAttributeBlendType=UseA}, False=@n20.MaterialAttributes) {group=Customization_Wrap, parameterName=UseIceBreakerWrap})
  MaterialFunctionCall() {materialFunction=/Engine/Functions/Engine_MaterialFunctions02/WorldPositionOffset/ObjectScale}
  Multiply(A=ObjectRadius(), B=@n34)
  ObjectBounds()
  n73 = Constant()
  n74 = SetMaterialAttributes!(MaterialAttributes, BaseColor=@n73, Metallic, Specular=@n73, Roughness, Normal, EmissiveColor=@n35) {attributeSetTypes [
    #2
    #3
    #4
    #5
    #1
    #6
   ]}
  Reroute(None=@n74)
  ShadingPathSwitch(Default=@n58, Deferred=TextureCoordinate()) {desc=Migrated from FeatureLevelSwitch}
 ]}
