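# Material function for the KevinCouture weapon wrap, blended over incoming material attributes by a customization mask behind a static switch.
# Top/side-projects color, normal and mask textures from pre-skinned bounds and normals, with separate UV transforms for weapons, vehicles and mechs.
# Emissive combines a sine-pulsed mask with a height-lerped gradient.
# A Blueprint-fed timestamp triggers a fire-reactive boost that adds the emissive texture, brightens the gradient, then decays.
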MaterialFunction /Game/Weapons/WeaponSkins/Wraps/MaterialFunctions/MF_KevinCouture_Wrap
graph {nodes [
  n1 = Sine(None=Multiply(A=Time(), B=ScalarParameter() {defaultValue=0.1, group=KevinCouture_Emissive, parameterName=KC-Emissive01-BrightnessTime, sortPriority=-10}))
  Clamp(
   None=Divide(A=Add(A=@n1))
   Min=ScalarParameter() {group=KevinCouture_Emissive, parameterName=KC-Emissive01-MinBrightness, sortPriority=-9}
   Max=ScalarParameter() {defaultValue=1, group=KevinCouture_Emissive, parameterName=KC-Emissive01-MaxBrightness, sortPriority=-8}
  )
  n2 = FunctionInput() {inputName=Material, inputType=FunctionInput_MaterialAttributes, sortPriority=-10}
  n3 = GetMaterialAttributes(None=@n2) {attributeGetTypes [#1, #2, #3]}
  n4 = TextureObjectParameter() {group=KevinCouture, parameterName=KC_BaseColor_01, sortPriority=11, texture=&T_BlacksRGB}
  n5 = FunctionInput(Preview=Constant()) {bUsePreviewValueAsDefault=true, inputName=isMech, inputType=FunctionInput_Scalar, sortPriority=4}
  n6 = LinearInterpolate(
   A=ScalarParameter() {defaultValue=1, group=KevinCouture, parameterName=KC_Blend_Contrast, sortPriority=-1}
   B=ScalarParameter() {defaultValue=2, group=KevinCouture, parameterName=KC_Blend_Contrast_Mech, sortPriority=-1}
   Alpha=@n5
  )
  n7 = ScalarParameter() {group=KevinCouture, parameterName=KC_Fill_alternate_Axis_Opacity 01, sortPriority=2}
  n8 = VectorParameter() {group=KevinCouture, parameterName="KC_Fill_alternate_Axis_Color_A=Roughness 01", sortPriority=3}
  n9 = VectorParameter() {group=KevinCouture_UVS, parameterName=KC-Rotation-Weapon-01---------------------, sortPriority=1, defaultValue=<rgba 0 0 -0.25 1>}
  n10 = VectorParameter() {group=KevinCouture_UVS, parameterName=KC_Rotation_Vehicle_01, sortPriority=2, defaultValue=<rgba 0 0 0 1>}
  n11 = VectorParameter() {
   desc=must activate TS_Vehicle_Side_01 UVs
   group=KevinCouture_UVS
   parameterName=KC_Rotation_Vehicle_Side_01
   sortPriority=2
   defaultValue=<rgba 0 0 0 1>
  }
  n12 = FunctionInput() {bUsePreviewValueAsDefault=true, inputName=Side Project Vehicle, inputType=FunctionInput_Scalar, sortPriority=6}
  n13 = Multiply(A=@n12, B=ScalarParameter() {group=KevinCouture_UVS, parameterName="KC_Vehicle_Side_01 UVs 0=Off 1=Active"})
  n14 = FunctionInput(Preview=Constant()) {inputName=isVehicle, inputType=FunctionInput_Scalar, sortPriority=5}
  n15 = VectorParameter() {group=KevinCouture_UVS, parameterName=KC_Rotation_Mech_01, sortPriority=3, defaultValue=<rgba 0.25 0.25 0 1>}
  n16 = LinearInterpolate(A=LinearInterpolate(A=@n9.RGB, B=LinearInterpolate(A=@n10.RGB, B=@n11.RGB, Alpha=@n13), Alpha=@n14), B=@n15.RGB, Alpha=@n5)
  n17 = FunctionInput(Preview=VertexInterpolator(VS=PreSkinnedNormal()).PS) {bUsePreviewValueAsDefault=true, inputName=Pre-Skinned Local Normal, sortPriority=10}
  n18 = VectorParameter() {group=KevinCouture_UVS, parameterName=KC_Offset_Weapon_01---------------------, sortPriority=10, defaultValue=<rgba 0 0 0 1>}
  n19 = VectorParameter() {group=KevinCouture_UVS, parameterName=KC_Offset_Vehicle_01, sortPriority=11, defaultValue=<rgba 0 0 0 1>}
  n20 = VectorParameter() {
   desc=must activate TS_Vehicle_Side_01 UVs
   group=KevinCouture_UVS
   parameterName=KC_Offset_Vehicle_Side_01
   sortPriority=11
   defaultValue=<rgba 0 0 0 1>
  }
  n21 = VectorParameter() {group=KevinCouture_UVS, parameterName=KC_Offset_Mech_01, sortPriority=12, defaultValue=<rgba 0 0 0 1>}
  n22 = LinearInterpolate(A=LinearInterpolate(A=@n18.RGB, B=LinearInterpolate(A=@n19.RGB, B=@n20.RGB, Alpha=@n13), Alpha=@n14), B=@n21.RGB, Alpha=@n5)
  n23 = PreSkinnedPosition()
  n24 = MaterialFunctionCall!(
   is Vehicle=@n14
   Object Bounds Max Adjustment=FunctionInput() {bUsePreviewValueAsDefault=true, inputName=Object Bounds Max Adjustment, sortPriority=13}
   Object Bounds Min Adjustment=FunctionInput() {bUsePreviewValueAsDefault=true, inputName=Object Bounds Min Adjustment, sortPriority=12}
   Object Bounds Center Adjustment=FunctionInput() {bUsePreviewValueAsDefault=true, inputName=Object Bounds Center Adjustment, sortPriority=14}
   isMech=@n5
   Side Project Vehicle
   Local Position=FunctionInput(Preview=VertexInterpolator(VS=@n23).PS) {bUsePreviewValueAsDefault=true, inputName=Local Position, sortPriority=11}
  ) {materialFunction=&/MF_WrapBounds}
  n25 = Time()
  n26 = VectorParameter() {group=KevinCouture_UVS, parameterName=KC_Panning_Weapon_01---------------------, sortPriority=40, defaultValue=<rgba 0 0 0 1>}
  n27 = VectorParameter() {group=KevinCouture_UVS, parameterName=KC_Panning_Vehicle_01, sortPriority=41, defaultValue=<rgba 0 0 0 1>}
  n28 = VectorParameter() {
   desc=must activate TS_Vehicle_Side_01 UVs
   group=KevinCouture_UVS
   parameterName=KC_Panning_Vehicle_Side_01
   sortPriority=41
   defaultValue=<rgba 0 0 0 1>
  }
  n29 = VectorParameter() {group=KevinCouture_UVS, parameterName=KC_Panning_Mech_01, sortPriority=42, defaultValue=<rgba 0 0 0 1>}
  n30 = LinearInterpolate(A=LinearInterpolate(A=@n26.RGB, B=LinearInterpolate(A=@n27.RGB, B=@n28.RGB, Alpha=@n13), Alpha=@n14), B=@n29.RGB, Alpha=@n5)
  n31 = VectorParameter() {group=KevinCouture_UVS, parameterName="KC_Scale_Weapon_01 A=uniform", sortPriority=30, defaultValue=<rgba 1 1 1 1>}
  n32 = Multiply(A=@n31.RGB, B=@n31.RGB)
  n33 = VectorParameter() {group=KevinCouture_UVS, parameterName="KC_Scale_Vehicle_01 A=uniform", sortPriority=31, defaultValue=<rgba 1 1 1 1>}
  n34 = VectorParameter() {
   desc=must activate TS_Vehicle_Side_01 UVs
   group=KevinCouture_UVS
   parameterName="KC_Scale_Vehicle_Side_01 A=uniform"
   sortPriority=31
   defaultValue=<rgba 1 1 1 1>
  }
  n35 = LinearInterpolate(A=Multiply(A=@n33.RGB, B=@n33.RGB), B=Multiply(A=@n34.RGB, B=@n34.RGB), Alpha=@n13)
  n36 = VectorParameter() {group=KevinCouture_UVS, parameterName="KC_Scale_Mech_01 A=uniform", defaultValue=<rgba 1 1 1 1>}
  n37 = Multiply(
   A=Add(A=Add(A=@n24.Result, B=@n22), B=Multiply(A=@n25, B=@n30))
   B=LinearInterpolate(A=LinearInterpolate(A=@n32, B=@n35, Alpha=@n14), B=Multiply(A=@n36.RGB, B=@n36.RGB), Alpha=@n5)
  )
  n38 = MaterialFunctionCall!(
   IsCharacter
   IsStatsB
   Texture=@n4
   Rotation_Axis=@n16
   Caculated_Bounds=@n37
   Pre_Skinned_Normals=@n17
   Blend_Contrast=@n6
   Fill_alternate_Axis_Opacity=@n7
   "Fill_alternate_Axis_Color_A=Roughness"=AppendVector(A=@n8.RGB, B=@n8.RGB)
   Is_Vehicle=@n14
   Is_Mech=@n5
   Enable_Fill_Z_Weapon
   Fill_Z_Weapon_opacity
   Fill_Z_Weapon_Color
  ) {materialFunction=&/MF_TopsSideProjection_Base}
  n39 = TextureObjectParameter() {
   group=KevinCouture
   parameterName=KC_Normal_01
   samplerType=SAMPLERTYPE_Normal
   sortPriority=16
   texture=&FlatNormalBlue
  }
  n40 = LinearInterpolate(
   A=@n3.Normal
   B=Add(
    A=@n3.Normal
    B=MaterialFunctionCall!(
     Texture_Normals=@n39
     Rotation_Axis=@n16
     Caculated_Bounds=@n37
     Pre_Skinned_Normals=@n17
     Blend_Contrast=@n6
     Fill_alternate_Axis_Opacity=@n7
     Is_Vehicle=@n14
     Is_Mech=@n5
    ) {materialFunction=&/MF_TopsSideProjection_Normals}.Add_Surface_Normal
   )
   Alpha=ScalarParameter() {defaultValue=1, group=KevinCouture, parameterName=KC_Normal_Intensty, sortPriority=1}
  )
  n41 = VectorParameter() {group=KevinCouture, parameterName=KC_Fill_alternate_Axis_Color_Mask_01, sortPriority=21, defaultValue=<rgba 0 0 0 1>}
  n42 = AppendVector(A=@n41.RGB, B=@n41.RGB)
  n43 = MaterialFunctionCall(
   Texture=TextureObjectParameter() {group=KevinCouture, parameterName=KC_Mask_01, sortPriority=20, texture=&T_BlacksRGB}
   Rotation_Axis=@n16
   Caculated_Bounds=@n37
   Pre_Skinned_Normals=@n17
   Blend_Contrast=@n6
   Fill_alternate_Axis_Opacity=@n7
   "Fill_alternate_Axis_Color_A=Roughness"=@n42
   Is_Vehicle=@n14
   Is_Mech=@n5
  ) {materialFunction=&MF_TopsSideProjection_Base}
  n44 = Multiply(A=Time(), B=ScalarParameter() {defaultValue=0.1, group=KevinCouture_Emissive, parameterName=KC-Emissive01-TimeMultiplier, sortPriority=9})
  n45 = Divide(A=Add(A=Sine(None=@n44)))
  n46 = ScalarParameter() {defaultValue=0.3, group=KevinCouture_Emissive, parameterName=KC-Emissive01-Min Brightness, sortPriority=11}
  n47 = ScalarParameter() {defaultValue=1, group=KevinCouture_Emissive, parameterName=KC-Emissive01-Max Brightness, sortPriority=12}
  n48 = TextureObjectParameter() {
   group=KevinCouture
   parameterName=KC_Mask_02
   samplerType=SAMPLERTYPE_Masks
   sortPriority=62
   texture=&T_BlackMask
  }
  n49 = VectorParameter() {group=KevinCouture, parameterName=KC_Fill_alternate_Axis_Color_Mask 02, sortPriority=67, defaultValue=<rgba 0 0 0 1>}
  n50 = VectorParameter() {group=KevinCouture_UVS, parameterName=KC-Rotation-Weapon-02---------------------, sortPriority=150, defaultValue=<rgba 0 0 -0.25 1>}
  n51 = VectorParameter() {group=KevinCouture_UVS, parameterName=KC_Rotation_Vehicle_02, sortPriority=151, defaultValue=<rgba 0 0 0 1>}
  n52 = VectorParameter() {
   desc=must activate TS_Vehicle_Side_02 UVs
   group=KevinCouture_UVS
   parameterName=KC_Rotation_Vehicle_Side_02
   sortPriority=151
   defaultValue=<rgba 0 0 0 1>
  }
  n53 = Multiply(A=@n12, B=ScalarParameter() {group=KevinCouture_UVS, parameterName="KC_Vehicle_Side_02 UVs  0=Off 1=Active", sortPriority=149})
  n54 = LinearInterpolate(A=@n51.RGB, B=@n52.RGB, Alpha=@n53)
  n55 = VectorParameter() {group=KevinCouture_UVS, parameterName=KC_Rotation_Mech_02, sortPriority=153, defaultValue=<rgba 0.25 0.25 0 1>}
  n56 = VectorParameter() {group=KevinCouture_UVS, parameterName=KC_Offset_Weapon_02---------------------, sortPriority=160, defaultValue=<rgba 0 0 0 1>}
  n57 = VectorParameter() {group=KevinCouture_UVS, parameterName=KC_Offset_Vehicle_02, sortPriority=161, defaultValue=<rgba 0 0 0 1>}
  n58 = VectorParameter() {
   desc=must activate TS_Vehicle_Side_02 UVs
   group=KevinCouture_UVS
   parameterName=KC_Offset_Vehicle_Side_02
   sortPriority=161
   defaultValue=<rgba 0 0 0 1>
  }
  n59 = LinearInterpolate(A=@n57.RGB, B=@n58.RGB, Alpha=@n53)
  n60 = VectorParameter() {group=KevinCouture_UVS, parameterName=KC_Offset_Mech_02, sortPriority=162, defaultValue=<rgba 0 0 0 1>}
  n61 = Add(A=@n24.Result, B=LinearInterpolate(A=LinearInterpolate(A=@n56.RGB, B=@n59, Alpha=@n14), B=@n60.RGB, Alpha=@n5))
  n62 = VectorParameter() {group=KevinCouture_UVS, parameterName=KC_Panning_Weapon_02---------------------, sortPriority=180, defaultValue=<rgba 0 0 0 1>}
  n63 = VectorParameter() {group=KevinCouture_UVS, parameterName=KC_Panning_Vehicle_02, sortPriority=181, defaultValue=<rgba 0 0 0 1>}
  n64 = VectorParameter() {
   desc=must activate TS_Vehicle_Side_02 UVs
   group=KevinCouture_UVS
   parameterName=KC_Panning_Vehicle_Side_02
   sortPriority=181
   defaultValue=<rgba 0 0 0 1>
  }
  n65 = LinearInterpolate(A=@n63.RGB, B=@n64.RGB, Alpha=@n53)
  n66 = VectorParameter() {group=KevinCouture_UVS, parameterName=KC_Panning_Mech_02, sortPriority=182, defaultValue=<rgba 0 0 0 1>}
  n67 = Multiply(A=Time(), B=LinearInterpolate(A=LinearInterpolate(A=@n62.RGB, B=@n65, Alpha=@n14), B=@n66.RGB, Alpha=@n5))
  n68 = VectorParameter() {group=KevinCouture_UVS, parameterName="KC_Scale_Weapon_02 A=uniform", sortPriority=170, defaultValue=<rgba 1 1 1 1>}
  n69 = VectorParameter() {group=KevinCouture_UVS, parameterName="KC_Scale_Vehicle_02 A=uniform", sortPriority=171, defaultValue=<rgba 1 1 1 1>}
  n70 = VectorParameter() {
   desc=must activate TS_Vehicle_Side_02 UVs
   group=KevinCouture_UVS
   parameterName="KC_Scale_Vehicle_Side_02 A=uniform"
   sortPriority=171
   defaultValue=<rgba 1 1 1 1>
  }
  n71 = LinearInterpolate(
   A=Multiply(A=@n68.RGB, B=@n68.RGB)
   B=LinearInterpolate(A=Multiply(A=@n69.RGB, B=@n69.RGB), B=Multiply(A=@n70.RGB, B=@n70.RGB), Alpha=@n53)
   Alpha=@n14
  )
  n72 = VectorParameter() {group=KevinCouture_UVS, parameterName="KC_Scale_Mech_02 A=uniform", sortPriority=172, defaultValue=<rgba 1 1 1 1>}
  n73 = Multiply(A=@n72.RGB, B=@n72.RGB)
  n74 = MaterialFunctionCall(
   Texture=@n48
   Rotation_Axis=LinearInterpolate(A=LinearInterpolate(A=@n50.RGB, B=@n54, Alpha=@n14), B=@n55.RGB, Alpha=@n5)
   Caculated_Bounds=Multiply(A=Add(A=@n61, B=@n67), B=LinearInterpolate(A=@n71, B=@n73, Alpha=@n5))
   Pre_Skinned_Normals=@n17
   Blend_Contrast=LinearInterpolate(
    A=ScalarParameter() {defaultValue=1, group=KevinCouture, parameterName=KC_Blend_Contrast, sortPriority=-1}
    B=ScalarParameter() {defaultValue=2, group=KevinCouture, parameterName=KC_Blend_Contrast_Mech, sortPriority=-1}
    Alpha=@n5
   )
   Fill_alternate_Axis_Opacity=ScalarParameter() {group=KevinCouture, parameterName=KC_Fill_alternate_Axis_Opacity 02, sortPriority=51}
   "Fill_alternate_Axis_Color_A=Roughness"=AppendVector(A=@n49.RGB, B=@n49.RGB)
   Is_Vehicle=@n14
   Is_Mech=@n5
  ) {materialFunction=&MF_TopsSideProjection_Base}
  n75 = ChannelMaskParameter(None=@n74.Combined) {group=KevinCouture_Emissive, parameterName=KC_Emissive02_MaskChannel_02, sortPriority=105}
  n76 = VectorParameter() {group=KevinCouture_Emissive, parameterName=KC_Emissive01_Color, sortPriority=8, defaultValue=<rgba 0 0 0 1>}
  n77 = Multiply(A=Multiply(A=Clamp(None=@n45, Min=@n46, Max=@n47) {minDefault=0.3}, B=@n75.R), B=@n76.RGB)
  n78 = VectorParameter() {group=KevinCouture_Emissive, parameterName=KC_Emissive02_Color, sortPriority=102, defaultValue=<rgba 0 0 0 1>}
  n79 = ScalarParameter() {defaultValue=1, group=KevinCouture_Emissive, parameterName=KC_Blend_Contrast, sortPriority=-1}
  n80 = MaterialFunctionCall(In=Abs(None=ComponentMask(None=@n17) {b=true}), Contrast=@n79) {materialFunction=&CheapContrast}
  n81 = MaterialFunctionCall(In=Abs(None=ComponentMask(None=@n17) {g=true}), Contrast=@n79) {materialFunction=&CheapContrast}
  n82 = Saturate(None=Subtract(A=@n80.Result, B=@n81.Result))
  n83 = MaterialFunctionCall(In=Abs(None=ComponentMask(None=@n17) {r=true}), Contrast=@n79) {materialFunction=&CheapContrast}
  n84 = Add(
   A=Saturate(None=Add(A=@n82, B=@n83.Result))
   B=ScalarParameter() {defaultValue=1, group=KevinCouture_Emissive, parameterName=KC_AEmissive_alternate_Axis_Masking_Weapon, sortPriority=3}
  )
  n85 = Saturate(None=Subtract(A=@n80.Result, B=@n83.Result))
  n86 = Add(
   A=Saturate(None=Add(A=@n85, B=@n81.Result))
   B=ScalarParameter() {defaultValue=1, group=KevinCouture_Emissive, parameterName=KC_AEmissive_alternate_Axis_Masking_Vehicle, sortPriority=3}
  )
  n87 = Saturate(None=LinearInterpolate(A=LinearInterpolate(A=@n84, B=@n86, Alpha=@n14), Alpha=@n5))
  n88 = Time()
  n89 = ScalarParameter() {group=00-Reactive-GTFB, parameterName=00-GameTimeFromBP}
  n90 = Saturate(None=Divide(
    A=Subtract(A=@n88, B=@n89)
    B=ScalarParameter() {defaultValue=1, group=KevinCouture_Reactive, parameterName=KC_AnimationDuration, sortPriority=33}
   ))
  n91 = ScalarParameter() {defaultValue=0.5, group=KevinCouture_Reactive, parameterName=KC_Curve_Amount, sortPriority=34}
  n92 = ScalarParameter() {group=KevinCouture_Reactive, parameterName=KC_TestReactive, sortPriority=300}
  n93 = LinearInterpolate(A=OneMinus(None=Power(Base=@n90, Exp=@n91)) {desc=inverts curve from 0-1 to 1-0}, B=@n92, Alpha=@n92)
  n94 = AppendVector(
   A=LinearInterpolate(B=ScalarParameter() {defaultValue=0.3, group=KevinCouture_Reactive, parameterName=KC_Time_Glow_SpeedAmount, sortPriority=603}, Alpha=@n93) {constB=0.5}
   B=Constant()
  )
  n95 = VectorParameter() {group=KevinCouture_Emissive, parameterName=KC_Emissive02_OffsetWeapon, sortPriority=124, defaultValue=<rgba 0 -1 0 0>}
  n96 = ScalarParameter() {defaultValue=0.5, group=KevinCouture_Emissive, parameterName="KC_Emissive02_FalloffWeapon (negative inverts)", sortPriority=123}
  n97 = VectorParameter() {group=KevinCouture_Emissive, parameterName=KC_Emissive02_OffsetVehicle, sortPriority=125, defaultValue=<rgba -1 0 0 0>}
  n98 = VectorParameter() {group=KevinCouture_Emissive, parameterName=KC_Emissive02_OffsetMech, sortPriority=126, defaultValue=<rgba 0 0 -1 0>}
  n99 = VectorParameter() {group=KevinCouture_Emissive, parameterName=KC_Emissive02_Spread_XY, sortPriority=120, defaultValue=<rgba 1 1 1 0>}
  n100 = VectorParameter() {group=KevinCouture_Emissive, parameterName=KC_Emissive02_Spread_Mech_XY, sortPriority=119, defaultValue=<rgba 1 1 1 0>}
  n101 = VectorParameter() {group=KevinCouture_Emissive, parameterName=KC_Emissive02_Animated_WeaponXY, sortPriority=130}
  n102 = VectorParameter() {group=KevinCouture_Emissive, parameterName=KC_Emissive02_Animated_VehicleXY, sortPriority=131}
  n103 = MaterialFunctionCall!(
   ForwardAxis=FunctionInput(Preview=Constant3Vector() {constant=<rgba 1 0 0 0>}) {bUsePreviewValueAsDefault=true, inputName=ForwardAxis}
   Bounds=@n24.Result
   Pre-Skinned Local Position
   isVehicle=@n14
   isMech=@n5
   Side Project Vehicle=@n12
   WeaponOffset=@n95.RGB
   WeaponFalloff=@n96
   VehicleOffset=@n97.RGB
   VehicleFalloff=ScalarParameter() {defaultValue=0.5, group=KevinCouture_Emissive, parameterName="KC_Emissive02_FalloffVehicle ", sortPriority=124}
   MechOffset=@n98.RGB
   MechFalloff=ScalarParameter() {defaultValue=0.5, group=KevinCouture_Emissive, parameterName=KC_Emissive02_FalloffMech, sortPriority=125}
   Invert_Gradent=StaticBoolParameter() {group=KevinCouture_Emissive, parameterName=KC_Emissive02_Invert_Gradent, sortPriority=110}
   UV_Spread_XY=LinearInterpolate(A=@n99.RGB, B=@n100.RGB, Alpha=@n5)
   UV_Animated_XY=LinearInterpolate(
    A=LinearInterpolate(A=@n101.RGB, B=@n102.RGB, Alpha=@n14)
    B=VectorParameter() {group=KevinCouture_Emissive, parameterName=KC_Emissive02_Animated_MechXY, sortPriority=132}.RGB
    Alpha=@n5
   )
   Warp_UVs=Multiply(
    A=ChannelMaskParameter(None=@n74.Combined) {group=KevinCouture_Emissive, parameterName=KC_WarpEmissive02MaskChannel_02, sortPriority=137}.R
    B=ScalarParameter() {group=KevinCouture_Emissive, parameterName=KC_WarpEmissive02_Warp_Amount, sortPriority=138}
   )
   Rotate_Mech=ScalarParameter() {defaultValue=1, group=KevinCouture_Emissive, parameterName=KC_Emissive02_Texture_Rotate_Mech, sortPriority=113}
   Rotate_Vehicle_Side=ScalarParameter() {defaultValue=1, group=KevinCouture_Emissive, parameterName=KC_Emissive02_Texture_Rotate_Vehicle_Side, sortPriority=113}
   Rotate_Vehicle=ScalarParameter() {defaultValue=1, group=KevinCouture_Emissive, parameterName=KC_Emissive02_Texture_Rotate_Vehicle, sortPriority=113}
  ) {materialFunction=&/MF_Wrap_Gradent_Creation}
  n104 = TextureObjectParameter() {
   group=KevinCouture_Emissive
   parameterName=KC_Emissive02_Gradent_Texture
   samplerType=SAMPLERTYPE_Masks
   texture=/Game/Global/Textures/Default/Masks_and_Gradients/T_VertGradient_MASK
  }
  n105 = ChannelMaskParameter(None=TextureSample(UVs=Add(A=@n94, B=@n103.Uvs), Tex=@n104) {samplerType=SAMPLERTYPE_Masks}.RGBA) {group=KevinCouture_Emissive, parameterName=KC_Emissive02_TextureChannel, sortPriority=110}
  n106 = Multiply(
   A=@n78.RGB
   B=MaterialFunctionCall(
    Transition Phase=Multiply(A=@n87, B=@n105.R)
    Height Texture=@n75.R
    Contrast=ScalarParameter() {group=KevinCouture_Emissive, parameterName=KC_Emissive02_HeightLerp_Contrast, sortPriority=108}
   ) {materialFunction=&HeightLerp}.Alpha
  )
  n107 = Multiply(A=@n106, B=ScalarParameter() {defaultValue=0.1, group=KevinCouture_Emissive, parameterName=KC-Emissive01-RuneBrightness, sortPriority=10})
  n108 = TextureObjectParameter() {group=KevinCouture, parameterName=KC_Emissive_01, sortPriority=11, texture=&T_BlacksRGB}
  n109 = MaterialFunctionCall(
   Texture=@n108
   Rotation_Axis=@n16
   Caculated_Bounds=@n37
   Pre_Skinned_Normals=@n17
   Blend_Contrast=@n6
   Fill_alternate_Axis_Opacity=Constant()
   "Fill_alternate_Axis_Color_A=Roughness"=@n42
   Is_Vehicle=@n14
   Is_Mech=@n5
  ) {materialFunction=&MF_TopsSideProjection_Base}
  n110 = Add(
   A=Multiply(
    A=@n109.RGB
    B=ScalarParameter() {defaultValue=2, group=KevinCouture_Reactive, parameterName=KC-Emissive01-CurveBrightness_Reactive, sortPriority=4}
   )
   B=@n106
  )
  n111 = SetMaterialAttributes!(
   MaterialAttributes=@n3.MaterialAttributes
   BaseColor=ComponentMask(None=@n38.Combined) {b=true, g=true, r=true}
   Roughness=LinearInterpolate(
    A=ScalarParameter() {group=KevinCouture, parameterName=KC_Roughness_Min, sortPriority=13}
    B=ScalarParameter() {defaultValue=1, group=KevinCouture, parameterName=KC_Roughness_Max, sortPriority=14}
    Alpha=ComponentMask(None=@n38.Combined) {a=true}
   )
   Normal=ShadingPathSwitch(Default=@n40, Deferred=@n40, Mobile=@n3.Normal) {desc=Migrated from FeatureLevelSwitch}
   Specular=LinearInterpolate(
    A=ScalarParameter() {group=KevinCouture, parameterName=KC_Specular_Min, sortPriority=12}
    B=ScalarParameter() {defaultValue=1, group=KevinCouture, parameterName=KC_Specular_Max, sortPriority=12}
    Alpha=ChannelMaskParameter(None=@n43.Combined) {group=KevinCouture, parameterName=KC_SpecChannel_Mask_01, sortPriority=26}.R
   )
   Metallic=ChannelMaskParameter(None=@n43.Combined) {group=KevinCouture, parameterName=KC_MetalicChannel_Mask_01, sortPriority=26}.R
   EmissiveColor=LinearInterpolate(A=Add(A=@n77, B=@n107), B=Add(A=@n110, B=@n77), Alpha=@n93)
  ) {attributeSetTypes [
    #1
    #2
    #3
    #4
    #5
    #6
   ]}
  n112 = StaticSwitchParameter(
   True=BlendMaterialAttributes(
    A=@n3.MaterialAttributes
    B=@n111
    Alpha=FunctionInput() {bUsePreviewValueAsDefault=true, inputName="Customization Mask (_CM B)", inputType=FunctionInput_Scalar, sortPriority=24}
   ) {vertexAttributeBlendType=UseA}
   False=@n2
  ) {
   desc="TopSide Projection Base - Ambient emissive  - Reactive -Topside Emissive with Texture and Ambient becomes brighter Height Lerp on Fire"
   group=Customization_Wrap
   parameterName=Wrap_KevinCouture
  }
  FunctionOutput(None=@n112)
  Multiply(A=@n109.RGB, B=ScalarParameter() {defaultValue=1, group=KevinCouture_Emissive, parameterName=KC-Emissive01-Brightness, sortPriority=-8})
 ]}
