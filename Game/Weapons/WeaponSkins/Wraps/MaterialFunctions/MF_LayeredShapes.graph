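# Material function for wraps that overlays two textured shape layers (base color, alpha mask) onto incoming material attributes, behind a LayeredShapes static switch.
# Each layer projects via MF_TopsSideProjection_Base or MF_Projection_Side_Normal, with rotation, offset, scale and time panning selected per weapon, vehicle or mech by IsState.
# Layer masks blend the layers' attributes and can dim emissive.
# Their sum, optionally limited by top-side masks and the customization mask, composites over the input.
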MaterialFunction /Game/Weapons/WeaponSkins/Wraps/MaterialFunctions/MF_LayeredShapes
graph {nodes [
  n1 = SetMaterialAttributes!(MaterialAttributes, BaseColor, Roughness, Metallic, Specular, EmissiveColor=Constant3Vector() {constant=<rgba 1 0 0.0661 0>}) {attributeSetTypes [
    #1
    #2
    #3
    #4
    #5
   ]}
  n2 = FunctionInput(Preview=@n1) {inputName=Material, inputType=FunctionInput_MaterialAttributes, sortPriority=-100}
  n3 = GetMaterialAttributes(None=@n2) {attributeGetTypes [#5]}
  n4 = MaterialFunctionCall(Float4=NamedRerouteUsage()) {materialFunction=&BreakOutFloat4Components}
  n5 = NamedRerouteUsage()
  n6 = Multiply(A=Time(), B=NamedRerouteUsage())
  n7 = TextureObjectParameter() {
   group=LayeredShapes
   parameterName="LS_BaseColor A=Mask_01-------------"
   sortPriority=20
   texture=/Game/Weapons/WeaponSkins/Wraps/Textures/FallLeaves/T_Wrap_FalllLeaves_leafC
  }
  n8 = TextureSample(
   UVs=MaterialFunctionCall!(
    For_Safe_Zones
    Normal_Map
    Caculated_Bounds=NamedRerouteUsage()
    Pre_Skinned_Normals=NamedRerouteUsage()
    Is Vehicle=@n4.R
    Is VehicleSide=@n4.R
    Is Mech=@n4.R
    Vehicle_Rotation
    UV_Rotation Bypass=NamedRerouteUsage()
    Mech_Rotation
    Uvs_Offset=Add(A=@n5, B=@n6)
    Offset-after rotation
    Uvs_TileRate=NamedRerouteUsage()
   ) {materialFunction=&/MF_Projection_Side_Normal}.UVs
   Tex=@n7
  )
  n9 = VectorParameter() {group=LayeredShapesUV, parameterName=LS-Rotation-Weapon-01---------------------, sortPriority=1, defaultValue=<rgba 0 0 -0.25 1>}
  n10 = VectorParameter() {group=LayeredShapesUV, parameterName=LS_Rotation_Vehicle_01, sortPriority=2, defaultValue=<rgba 0 0 0 1>}
  n11 = VectorParameter() {group=LayeredShapesUV, parameterName=LS_Rotation_Vehicle_Side_01, sortPriority=2, defaultValue=<rgba 0 0 0 1>}
  n12 = Constant4Vector()
  n13 = MaterialFunctionCall(Float4=NamedRerouteDeclaration(None=FunctionInput(Preview=@n12.RGBA) {inputName=IsState, inputType=FunctionInput_Vector4, sortPriority=-1}) {name=IsState, nodeColor=<rgba 1 0 0.4687 1>}) {materialFunction=&BreakOutFloat4Components}
  n14 = LinearInterpolate(A=@n10.RGB, B=@n11.RGB, Alpha=@n13.G)
  n15 = VectorParameter() {group=LayeredShapesUV, parameterName=LS_Rotation_Mech_01, sortPriority=3, defaultValue=<rgba 0.25 0.25 0 1>}
  n16 = VectorParameter() {group=LayeredShapesUV, parameterName=LS_Offset_Weapon_01---------------------, sortPriority=10, defaultValue=<rgba 0 0 0 1>}
  n17 = VectorParameter() {group=LayeredShapesUV, parameterName=LS_Offset_Vehicle_01, sortPriority=11, defaultValue=<rgba 0 0 0 1>}
  n18 = VectorParameter() {desc=LS_Offset_Vehicle_Side_01, group=LayeredShapesUV, parameterName=LS_Offset_Vehicle_Side_01, sortPriority=11, defaultValue=<rgba 0 0 0 1>}
  n19 = Multiply(A=@n13.G, B=ScalarParameter() {group=LayeredShapesUV, parameterName="LS_Vehicle_Side 01 UV 0=Off 1=Active"})
  n20 = LinearInterpolate(A=@n17.RGB, B=@n18.RGB, Alpha=@n19)
  n21 = VectorParameter() {group=LayeredShapesUV, parameterName=LS_Offset_Mech_01, sortPriority=12, defaultValue=<rgba 0 0 0 1>}
  n22 = PreSkinnedLocalBounds()
  n23 = Subtract(A=VertexInterpolator(VS=PreSkinnedPosition()).PS, B=@n22.Min)
  n24 = Subtract(A=@n22.Max, B=@n22.Max)
  n25 = NamedRerouteDeclaration(None=FunctionInput(Preview=Divide(A=@n23, B=@n24)) {bUsePreviewValueAsDefault=true, inputName=Pre-Skinned Bounds, sortPriority=10}) {name=Caculated_Bounds, nodeColor=<rgba 0 0.0391 1 1>}
  n26 = Add(A=@n25, B=LinearInterpolate(A=LinearInterpolate(A=@n16.RGB, B=@n20), B=@n21.RGB))
  n27 = VectorParameter() {group=LayeredShapesUV, parameterName="LS_Scale_Weapon_01 A=uniform-------------", sortPriority=30, defaultValue=<rgba 1 1 1 1>}
  n28 = VectorParameter() {group=LayeredShapesUV, parameterName="LS_Scale_Vehicle_01 A=uniform", sortPriority=31, defaultValue=<rgba 1 1 1 1>}
  n29 = Multiply(A=@n28.RGB, B=@n28.RGB)
  n30 = VectorParameter() {
   desc=LS_Offset_Vehicle_Side_01
   group=LayeredShapesUV
   parameterName="LS_Scale_Vehicle_Side_01 A=uniform"
   sortPriority=31
   defaultValue=<rgba 1 1 1 1>
  }
  n31 = Multiply(A=@n30.RGB, B=@n30.RGB)
  n32 = VectorParameter() {group=LayeredShapesUV, parameterName="LS_Scale_Mech_01 A=uniform", defaultValue=<rgba 1 1 1 1>}
  n33 = LinearInterpolate(
   A=LinearInterpolate(A=Multiply(A=@n27.RGB, B=@n27.RGB), B=LinearInterpolate(A=@n29, B=@n31, Alpha=@n19), Alpha=@n13.R)
   B=Multiply(A=@n32.RGB, B=@n32.RGB)
   Alpha=@n13.B
  )
  n34 = Time()
  n35 = VectorParameter() {group=LayeredShapesUV, parameterName=LS_Panning_Weapon_01---------------------, sortPriority=40, defaultValue=<rgba 0 0 0 1>}
  n36 = VectorParameter() {group=LayeredShapesUV, parameterName=LS_Panning_Vehicle_01, sortPriority=41, defaultValue=<rgba 0 0 0 1>}
  n37 = VectorParameter() {
   desc=LS_Offset_Vehicle_Side_01
   group=LayeredShapesUV
   parameterName=LS_Panning_Vehicle_Side_01
   sortPriority=41
   defaultValue=<rgba 0 0 0 1>
  }
  n38 = VectorParameter() {group=LayeredShapesUV, parameterName=LS_Panning_Mech_01, sortPriority=42, defaultValue=<rgba 0 0 0 1>}
  n39 = LinearInterpolate(A=LinearInterpolate(A=@n35.RGB, B=LinearInterpolate(A=@n36.RGB, B=@n37.RGB, Alpha=@n19), Alpha=@n13.R), B=@n38.RGB, Alpha=@n13.B)
  n40 = VectorParameter() {group=LayeredShapes, parameterName="LS_Fill_alternate_Axis_Color_A=Alpha 01", sortPriority=23}
  n41 = PreSkinnedNormal()
  n42 = NamedRerouteDeclaration(None=FunctionInput(Preview=VertexInterpolator(VS=@n41).PS) {bUsePreviewValueAsDefault=true, inputName=Pre-Skinned Local Normal, sortPriority=10}) {name=PreSkin_Normals, nodeColor=<rgba 0.0781 0 1 1>}
  n43 = MaterialFunctionCall!(
   IsCharacter
   IsStatsB
   Texture=@n7
   Rotation_Axis=LinearInterpolate(A=LinearInterpolate(A=@n9.RGB, B=@n14, Alpha=@n13.R), B=@n15.RGB, Alpha=@n13.B)
   Caculated_Bounds=Add(A=Multiply(A=@n26, B=@n33), B=Multiply(A=@n34, B=@n39))
   Pre_Skinned_Normals=@n42
   Blend_Contrast=ScalarParameter() {defaultValue=1, group=LayeredShapes, parameterName=LS_TriBlend_Contrast, sortPriority=-1}
   Fill_alternate_Axis_Opacity=ScalarParameter() {defaultValue=1, group=LayeredShapes, parameterName=LS_Fill_alternate_Axis_Opacity 01, sortPriority=22}
   "Fill_alternate_Axis_Color_A=Roughness"=AppendVector(A=@n40.RGB, B=@n40.RGB)
   Is_Vehicle=@n13.R
   Is_Mech=@n13.B
   Enable_Fill_Z_Weapon
   Fill_Z_Weapon_opacity
   Fill_Z_Weapon_Color
  ) {materialFunction=&/MF_TopsSideProjection_Base}
  n44 = StaticBoolParameter() {group=LayeredShapes, parameterName=LS_Use_SideProjection_01, sortPriority=19}
  n45 = Add(A=NamedRerouteUsage(), B=ScalarParameter() {defaultValue=1, group=LayeredShapes, parameterName="LS_TriMask_01  1=off 0=on", sortPriority=20})
  n46 = StaticSwitch(True=Multiply(A=@n8.A, B=Saturate(None=@n45)), False=@n43.Alpha, Value=@n44)
  n47 = MaterialFunctionCall(Float4=NamedRerouteUsage()) {materialFunction=&BreakOutFloat4Components}
  n48 = NamedRerouteUsage()
  n49 = Multiply(A=Time(), B=NamedRerouteUsage())
  n50 = TextureObjectParameter() {
   group=LayeredShapes
   parameterName="LS_BaseColor A=Mask_02-------------"
   sortPriority=40
   texture=/Game/Weapons/WeaponSkins/Wraps/Textures/FallLeaves/T_Wrap_FalllLeaves_leafA
  }
  n51 = StaticSwitchParameter(True=@n7, False=@n50) {group=LayeredShapes, parameterName=LS-Use-BaseColor-01-for-02--------------, sortPriority=39}
  n52 = TextureSample(
   UVs=MaterialFunctionCall(
    Caculated_Bounds=NamedRerouteUsage()
    Pre_Skinned_Normals=NamedRerouteUsage()
    Is Vehicle=@n47.R
    Is VehicleSide=@n47.R
    Is Mech=@n47.R
    UV_Rotation Bypass=NamedRerouteUsage()
    Uvs_Offset=Add(A=@n48, B=@n49)
    Uvs_TileRate=NamedRerouteUsage()
   ) {materialFunction=&MF_Projection_Side_Normal}.UVs
   Tex=@n51
  )
  n53 = Add(A=NamedRerouteUsage(), B=ScalarParameter() {defaultValue=1, group=LayeredShapes, parameterName="LS_TriMask_02  1=off 0=on", sortPriority=40})
  n54 = VectorParameter() {group=LayeredShapesUV, parameterName=LS-Rotation-Weapon-02---------------------, sortPriority=101, defaultValue=<rgba 0 0 -0.25 1>}
  n55 = VectorParameter() {group=LayeredShapesUV, parameterName=LS_Rotation_Vehicle_02, sortPriority=102, defaultValue=<rgba 0 0 0 1>}
  n56 = VectorParameter() {group=LayeredShapesUV, parameterName=LS_Rotation_Vehicle_Side_02, sortPriority=102, defaultValue=<rgba 0 0 0 1>}
  n57 = LinearInterpolate(A=@n55.RGB, B=@n56.RGB, Alpha=@n13.G)
  n58 = VectorParameter() {group=LayeredShapesUV, parameterName=LS_Rotation_Mech_02, sortPriority=103, defaultValue=<rgba 0.25 0.25 0 1>}
  n59 = VectorParameter() {group=LayeredShapesUV, parameterName=LS_Offset_Weapon_02---------------------, sortPriority=110, defaultValue=<rgba 0 0 0 1>}
  n60 = VectorParameter() {group=LayeredShapesUV, parameterName=LS_Offset_Vehicle_02, sortPriority=111, defaultValue=<rgba 0 0 0 1>}
  n61 = VectorParameter() {
   desc=LS_Vehicle_Side 02 UV must be active
   group=LayeredShapesUV
   parameterName=LS_Offset_Vehicle_Side_02
   sortPriority=111
   defaultValue=<rgba 0 0 0 1>
  }
  n62 = Multiply(A=@n13.G, B=ScalarParameter() {group=LayeredShapesUV, parameterName="LS_Vehicle_Side 02 UV 0=Off 1=Active", sortPriority=100})
  n63 = LinearInterpolate(A=@n60.RGB, B=@n61.RGB, Alpha=@n62)
  n64 = VectorParameter() {group=LayeredShapesUV, parameterName=LS_Offset_Mech_02, sortPriority=112, defaultValue=<rgba 0 0 0 1>}
  n65 = Add(A=@n25, B=LinearInterpolate(A=LinearInterpolate(A=@n59.RGB, B=@n63, Alpha=@n13.R), B=@n64.RGB, Alpha=@n13.B))
  n66 = VectorParameter() {group=LayeredShapesUV, parameterName="LS_Scale_Weapon_02 A=uniform-------------", sortPriority=130, defaultValue=<rgba 1 1 1 1>}
  n67 = VectorParameter() {group=LayeredShapesUV, parameterName="LS_Scale_Vehicle_02 A=uniform", sortPriority=131, defaultValue=<rgba 1 1 1 1>}
  n68 = Multiply(A=@n67.RGB, B=@n67.RGB)
  n69 = VectorParameter() {
   desc=LS_Offset_Vehicle_Side_02
   group=LayeredShapesUV
   parameterName="LS_Scale_Vehicle_Side_02 A=uniform"
   sortPriority=131
   defaultValue=<rgba 1 1 1 1>
  }
  n70 = Multiply(A=@n69.RGB, B=@n69.RGB)
  n71 = VectorParameter() {group=LayeredShapesUV, parameterName="LS_Scale_Mech_02 A=uniform", sortPriority=132, defaultValue=<rgba 1 1 1 1>}
  n72 = LinearInterpolate(
   A=LinearInterpolate(A=Multiply(A=@n66.RGB, B=@n66.RGB), B=LinearInterpolate(A=@n68, B=@n70, Alpha=@n62), Alpha=@n13.R)
   B=Multiply(A=@n71.RGB, B=@n71.RGB)
   Alpha=@n13.B
  )
  n73 = Time()
  n74 = VectorParameter() {group=LayeredShapesUV, parameterName=LS_Panning_Weapon_02---------------------, sortPriority=140, defaultValue=<rgba 0 0 0 1>}
  n75 = VectorParameter() {group=LayeredShapesUV, parameterName=LS_Panning_Vehicle_02, sortPriority=141, defaultValue=<rgba 0 0 0 1>}
  n76 = VectorParameter() {
   desc=LS_Offset_Vehicle_Side_02
   group=LayeredShapesUV
   parameterName=LS_Panning_Vehicle_Side_02
   sortPriority=141
   defaultValue=<rgba 0 0 0 1>
  }
  n77 = VectorParameter() {group=LayeredShapesUV, parameterName=LS_Panning_Mech_02, sortPriority=142, defaultValue=<rgba 0 0 0 1>}
  n78 = LinearInterpolate(A=LinearInterpolate(A=@n74.RGB, B=LinearInterpolate(A=@n75.RGB, B=@n76.RGB, Alpha=@n62), Alpha=@n13.R), B=@n77.RGB, Alpha=@n13.B)
  n79 = VectorParameter() {group=LayeredShapes, parameterName="LS_Fill_alternate_Axis_Color_A=Alpha 02", sortPriority=43}
  n80 = MaterialFunctionCall(
   Texture=@n51
   Rotation_Axis=LinearInterpolate(A=LinearInterpolate(A=@n54.RGB, B=@n57, Alpha=@n13.R), B=@n58.RGB, Alpha=@n13.B)
   Caculated_Bounds=Add(A=Multiply(A=@n65, B=@n72), B=Multiply(A=@n73, B=@n78))
   Pre_Skinned_Normals=@n42
   Blend_Contrast=ScalarParameter() {defaultValue=1, group=LayeredShapes, parameterName=LS_TriBlend_Contrast, sortPriority=-1}
   Fill_alternate_Axis_Opacity=ScalarParameter() {defaultValue=1, group=LayeredShapes, parameterName=LS_Fill_alternate_Axis_Opacity 02, sortPriority=42}
   "Fill_alternate_Axis_Color_A=Roughness"=AppendVector(A=@n79.RGB, B=@n79.RGB)
   Is_Vehicle=@n13.R
   Is_Mech=@n13.B
  ) {materialFunction=&MF_TopsSideProjection_Base}
  n81 = StaticBoolParameter() {group=LayeredShapes, parameterName=LS_Use_SideProjection_02, sortPriority=39}
  n82 = StaticSwitch(True=Multiply(A=@n52.A, B=Saturate(None=@n53)), False=@n80.Alpha, Value=@n81)
  n83 = Add(A=@n46, B=@n82)
  n84 = ScalarParameter() {group=LayeredShapes, parameterName="LS_Mask_Emissive 1=on", sortPriority=20}
  n85 = Saturate(None=OneMinus(None=Multiply(A=@n83, B=@n84) {constB=0}))
  n86 = SetMaterialAttributes(
   MaterialAttributes=@n3.MaterialAttributes
   BaseColor=StaticSwitch(True=@n8.RGB, False=@n43.RGB, Value=@n44)
   Roughness=ScalarParameter() {group=LayeredShapes, parameterName=LS_Roughness_01, sortPriority=21}
   Metallic=ScalarParameter() {group=LayeredShapes, parameterName=LS_Metalic01, sortPriority=21}
   Specular=ScalarParameter() {defaultValue=0.5, group=LayeredShapes, parameterName=LS_Specular01, sortPriority=21}
   EmissiveColor=Multiply(A=@n3.EmissiveColor, B=@n85)
  ) {attributeSetTypes [
    #1
    #2
    #3
    #4
    #5
   ]}
  n87 = GetMaterialAttributes(None=@n2) {attributeGetTypes [#5]}
  n88 = SetMaterialAttributes(
   MaterialAttributes=@n87.MaterialAttributes
   BaseColor=StaticSwitch(True=@n52.RGB, False=@n80.RGB, Value=@n81)
   Roughness=ScalarParameter() {group=LayeredShapes, parameterName=LS_Roughness_02, sortPriority=41}
   Metallic=ScalarParameter() {group=LayeredShapes, parameterName=LS_Metalic02, sortPriority=41}
   Specular=ScalarParameter() {group=LayeredShapes, parameterName=LS_Specular02, sortPriority=41}
   EmissiveColor=Multiply(A=@n85, B=@n87.EmissiveColor)
  ) {attributeSetTypes [
    #1
    #2
    #3
    #4
    #5
   ]}
  n89 = BlendMaterialAttributes(A=@n86, B=@n88, Alpha=@n82) {vertexAttributeBlendType=UseA}
  n90 = Saturate(None=Add(A=@n46, B=@n82))
  n91 = StaticSwitchParameter(
   True=ChannelMaskParameter(None=FunctionInput() {inputName=TS_Mask_02, inputType=FunctionInput_Vector4, sortPriority=20}) {group=LayeredShapes, parameterName=LS_Set_TS_Mask_02_Channel, sortPriority=2}.R
   False=ChannelMaskParameter(None=FunctionInput() {inputName=TS_Mask_01, inputType=FunctionInput_Vector4, sortPriority=20}) {group=LayeredShapes, parameterName=LS_Set_TS_Mask_01_Channel, sortPriority=2}.R
  ) {group=LayeredShapes, parameterName=Use MaskFrom TopSide 02, sortPriority=1}
  n92 = FunctionInput(Preview=Constant() {r=1}) {bUsePreviewValueAsDefault=true, inputName="Customization Mask (_CM B)", inputType=FunctionInput_Scalar, sortPriority=24}
  n93 = ScalarParameter() {group=LayeredShapes, parameterName="LS_Remove_Customization Mask ", sortPriority=-10}
  n94 = Multiply(
   A=StaticSwitchParameter(True=Multiply(A=@n90, B=@n91), False=@n90) {group=LayeredShapes, parameterName=Use MaskFrom TopSide 01}
   B=Saturate(None=Add(A=@n92, B=@n93))
  )
  FunctionOutput(None=StaticSwitchParameter(True=BlendMaterialAttributes(A=@n2, B=@n89, Alpha=@n94) {vertexAttributeBlendType=UseA}, False=@n2) {group=Customization, parameterName=LayeredShapes})
  NamedRerouteDeclaration(None=FunctionInput(Preview=Constant2Vector()) {bUsePreviewValueAsDefault=true, inputName=IsStateB, inputType=FunctionInput_Vector2, sortPriority=-2}) {name=IsStateB, nodeColor=<rgba 1 0 0.4688 1>}
  n95 = ScalarParameter() {group=LayeredShapesUV, parameterName=LS02_Rotation_Vehicle}
  n96 = ScalarParameter() {group=LayeredShapesUV, parameterName=LS02_Rotation_Vehicle_Side}
  n97 = MaterialFunctionCall(Float4=NamedRerouteUsage()) {materialFunction=&BreakOutFloat4Components}
  n98 = LinearInterpolate(
   A=LinearInterpolate(
    A=ScalarParameter() {group=LayeredShapesUV, parameterName=LS02_Rotation_Weapon}
    B=LinearInterpolate(A=@n95, B=@n96, Alpha=@n97.G)
    Alpha=@n97.R
   )
   B=LinearInterpolate(
    A=ScalarParameter() {group=LayeredShapesUV, parameterName=LS02_Rotation_Mech}
    B=ScalarParameter() {group=LayeredShapesUV, parameterName=LS02_Rotation_Character}
    Alpha=@n97.A
   )
   Alpha=@n97.B
  )
  n99 = ScalarParameter() {group=LayeredShapesUV, parameterName=LS02_Rotation_Backpack}
  n100 = MaterialFunctionCall(Float2=NamedRerouteUsage()) {materialFunction=&BreakOutFloat2Components}
  NamedRerouteDeclaration(None=LinearInterpolate(
    A=LinearInterpolate(A=@n98, B=@n99, Alpha=@n100.R)
    B=ScalarParameter() {group=LayeredShapesUV, parameterName=LS02_Rotation_Pickaxe}
    Alpha=@n100.G
   )) {name=UVs_Rotation_02, nodeColor=<rgba 0 0.7188 1 1>}
  n101 = ScalarParameter() {group=LayeredShapesUV, parameterName=LS01_Rotation_Vehicle}
  n102 = ScalarParameter() {group=LayeredShapesUV, parameterName=LS01_Rotation_Vehicle_Side}
  n103 = MaterialFunctionCall(Float4=NamedRerouteUsage()) {materialFunction=&BreakOutFloat4Components}
  n104 = LinearInterpolate(
   A=LinearInterpolate(
    A=ScalarParameter() {group=LayeredShapesUV, parameterName=LS01_Rotation_Weapon}
    B=LinearInterpolate(A=@n101, B=@n102, Alpha=@n103.G)
    Alpha=@n103.R
   )
   B=LinearInterpolate(
    A=ScalarParameter() {group=LayeredShapesUV, parameterName=LS01_Rotation_Mech}
    B=ScalarParameter() {group=LayeredShapesUV, parameterName=LS01_Rotation_Character}
    Alpha=@n103.A
   )
   Alpha=@n103.B
  )
  n105 = ScalarParameter() {group=LayeredShapesUV, parameterName=LS01_Rotation_Backpack}
  n106 = MaterialFunctionCall(Float2=NamedRerouteUsage()) {materialFunction=&BreakOutFloat2Components}
  NamedRerouteDeclaration(None=LinearInterpolate(
    A=LinearInterpolate(A=@n104, B=@n105, Alpha=@n106.R)
    B=ScalarParameter() {group=LayeredShapesUV, parameterName=LS01_Rotation_Pickaxe}
    Alpha=@n106.G
   )) {name=UVs_Rotation_01, nodeColor=<rgba 0 0.7188 1 1>}
  n107 = MaterialFunctionCall(Color=Abs(None=NamedRerouteUsage())) {materialFunction=&SplitComponents}
  n108 = MaterialFunctionCall(Float4=NamedRerouteUsage()) {materialFunction=&BreakOutFloat4Components}
  n109 = LinearInterpolate(A=@n107.R, B=LinearInterpolate(A=@n107.B, B=@n107.B, Alpha=@n108.G), Alpha=@n108.R)
  NamedRerouteDeclaration(None=MaterialFunctionCall(
    In=LinearInterpolate(A=@n109, B=@n107.G, Alpha=@n108.B)
    Contrast=ScalarParameter() {defaultValue=1, group=LayeredShapes, parameterName=LS_Tri_Blend_Contrast}
   ) {materialFunction=&CheapContrast}.Result) {name=Tri_Mask, nodeColor=<rgba 1 0.4688 0 1>}
  n110 = VectorParameter() {group=LayeredShapesUV, parameterName=LS02_Scale_Weapon, defaultValue=<rgba 1 1 1 1>}
  n111 = VectorParameter() {group=LayeredShapesUV, parameterName=LS02_Scale_Vehicle, defaultValue=<rgba 1 1 1 1>}
  n112 = Multiply(A=@n111.RGB, B=@n111.RGB)
  n113 = VectorParameter() {group=LayeredShapesUV, parameterName=LS02_Scale_VehicleSide, defaultValue=<rgba 1 1 1 1>}
  n114 = Multiply(A=@n113.RGB, B=@n113.RGB)
  n115 = MaterialFunctionCall(Float4=NamedRerouteUsage()) {materialFunction=&BreakOutFloat4Components}
  n116 = VectorParameter() {group=LayeredShapesUV, parameterName=LS02_Scale_Mech, defaultValue=<rgba 1 1 1 1>}
  n117 = VectorParameter() {group=LayeredShapesUV, parameterName=LS02_Scale_Character, defaultValue=<rgba 1 1 1 1>}
  n118 = LinearInterpolate(
   A=LinearInterpolate(A=Multiply(A=@n110.RGB, B=@n110.RGB), B=LinearInterpolate(A=@n112, B=@n114, Alpha=@n115.G), Alpha=@n115.R)
   B=LinearInterpolate(A=Multiply(A=@n116.RGB, B=@n116.RGB), B=Multiply(A=@n117.RGB, B=@n117.RGB), Alpha=@n115.A)
   Alpha=@n115.B
  )
  n119 = VectorParameter() {group=LayeredShapesUV, parameterName=LS02_Scale_Backpack, defaultValue=<rgba 1 1 1 1>}
  n120 = Multiply(A=@n119.RGB, B=@n119.RGB)
  n121 = MaterialFunctionCall(Float2=NamedRerouteUsage()) {materialFunction=&BreakOutFloat2Components}
  n122 = VectorParameter() {group=LayeredShapesUV, parameterName=LS02_Scale_Pickaxe, defaultValue=<rgba 1 1 1 1>}
  NamedRerouteDeclaration(None=LinearInterpolate(A=LinearInterpolate(A=@n118, B=@n120, Alpha=@n121.R), B=Multiply(A=@n122.RGB, B=@n122.RGB), Alpha=@n121.G)) {name=UVs_Scale02, nodeColor=<rgba 0 0.625 1 1>}
  n123 = VectorParameter() {group=LayeredShapesUV, parameterName=LS01_Offset_Vehicle, defaultValue=<rgba 0 0 0 1>}
  n124 = VectorParameter() {group=LayeredShapesUV, parameterName=LS01_Offset_VehicleSide, defaultValue=<rgba 0 0 0 1>}
  n125 = MaterialFunctionCall(Float4=NamedRerouteUsage()) {materialFunction=&BreakOutFloat4Components}
  n126 = LinearInterpolate(
   A=LinearInterpolate(
    A=VectorParameter() {group=LayeredShapesUV, parameterName=LS01_Offset_Weapon, defaultValue=<rgba 0 0 0 1>}.RGB
    B=LinearInterpolate(A=@n123.RGB, B=@n124.RGB, Alpha=@n125.G)
    Alpha=@n125.R
   )
   B=LinearInterpolate(
    A=VectorParameter() {group=LayeredShapesUV, parameterName=LS01_Offset_Mech, defaultValue=<rgba 0 0 0 1>}.RGB
    B=VectorParameter() {group=LayeredShapesUV, parameterName=LS01_Offset_Character, defaultValue=<rgba 0 0 0 1>}.RGB
    Alpha=@n125.A
   )
   Alpha=@n125.B
  )
  n127 = VectorParameter() {group=LayeredShapesUV, parameterName=LS01_Offset_Backpack, defaultValue=<rgba 0 0 0 1>}
  n128 = MaterialFunctionCall(Float2=NamedRerouteUsage()) {materialFunction=&BreakOutFloat2Components}
  NamedRerouteDeclaration(None=LinearInterpolate(
    A=LinearInterpolate(A=@n126, B=@n127.RGB, Alpha=@n128.R)
    B=VectorParameter() {group=LayeredShapesUV, parameterName=LS01_Offset_Pickaxe, defaultValue=<rgba 0 0 0 1>}.RGB
    Alpha=@n128.G
   )) {name=UVs_Offset_01, nodeColor=<rgba 0 0.625 1 1>}
  n129 = VectorParameter() {group=LayeredShapesUV, parameterName=LS02_Offset_Vehicle, defaultValue=<rgba 0 0 0 1>}
  n130 = VectorParameter() {group=LayeredShapesUV, parameterName=LS02_Offset_VehicleSide, defaultValue=<rgba 0 0 0 1>}
  n131 = MaterialFunctionCall(Float4=NamedRerouteUsage()) {materialFunction=&BreakOutFloat4Components}
  n132 = LinearInterpolate(
   A=LinearInterpolate(
    A=VectorParameter() {group=LayeredShapesUV, parameterName=LS02_Offset_Weapon, defaultValue=<rgba 0 0 0 1>}.RGB
    B=LinearInterpolate(A=@n129.RGB, B=@n130.RGB, Alpha=@n131.G)
    Alpha=@n131.R
   )
   B=LinearInterpolate(
    A=VectorParameter() {group=LayeredShapesUV, parameterName=LS02_Offset_Mech, defaultValue=<rgba 0 0 0 1>}.RGB
    B=VectorParameter() {group=LayeredShapesUV, parameterName=LS02_Offset_Character, defaultValue=<rgba 0 0 0 1>}.RGB
    Alpha=@n131.A
   )
   Alpha=@n131.B
  )
  n133 = VectorParameter() {group=LayeredShapesUV, parameterName=LS02_Offset_Backpack, defaultValue=<rgba 0 0 0 1>}
  n134 = MaterialFunctionCall(Float2=NamedRerouteUsage()) {materialFunction=&BreakOutFloat2Components}
  NamedRerouteDeclaration(None=LinearInterpolate(
    A=LinearInterpolate(A=@n132, B=@n133.RGB, Alpha=@n134.R)
    B=VectorParameter() {group=LayeredShapesUV, parameterName=LS02_Offset_Pickaxe, defaultValue=<rgba 0 0 0 1>}.RGB
    Alpha=@n134.G
   )) {name=UVs_Offset_02, nodeColor=<rgba 0 0.625 1 1>}
  n135 = VectorParameter() {group=LayeredShapesUV, parameterName=LS01_Scale_Weapon, defaultValue=<rgba 1 1 1 1>}
  n136 = VectorParameter() {group=LayeredShapesUV, parameterName=LS01_Scale_Vehicle, defaultValue=<rgba 1 1 1 1>}
  n137 = Multiply(A=@n136.RGB, B=@n136.RGB)
  n138 = VectorParameter() {group=LayeredShapesUV, parameterName=LS01_Scale_VehicleSide, defaultValue=<rgba 1 1 1 1>}
  n139 = Multiply(A=@n138.RGB, B=@n138.RGB)
  n140 = MaterialFunctionCall(Float4=NamedRerouteUsage()) {materialFunction=&BreakOutFloat4Components}
  n141 = VectorParameter() {group=LayeredShapesUV, parameterName=LS01_Scale_Mech, defaultValue=<rgba 1 1 1 1>}
  n142 = VectorParameter() {group=LayeredShapesUV, parameterName=LS01_Scale_Character, defaultValue=<rgba 1 1 1 1>}
  n143 = LinearInterpolate(
   A=LinearInterpolate(A=Multiply(A=@n135.RGB, B=@n135.RGB), B=LinearInterpolate(A=@n137, B=@n139, Alpha=@n140.G), Alpha=@n140.R)
   B=LinearInterpolate(A=Multiply(A=@n141.RGB, B=@n141.RGB), B=Multiply(A=@n142.RGB, B=@n142.RGB), Alpha=@n140.A)
   Alpha=@n140.B
  )
  n144 = VectorParameter() {group=LayeredShapesUV, parameterName=LS01_Scale_Backpack, defaultValue=<rgba 1 1 1 1>}
  n145 = Multiply(A=@n144.RGB, B=@n144.RGB)
  n146 = MaterialFunctionCall(Float2=NamedRerouteUsage()) {materialFunction=&BreakOutFloat2Components}
  n147 = VectorParameter() {group=LayeredShapesUV, parameterName=LS01_Scale_Pickaxe, defaultValue=<rgba 1 1 1 1>}
  NamedRerouteDeclaration(None=LinearInterpolate(A=LinearInterpolate(A=@n143, B=@n145, Alpha=@n146.R), B=Multiply(A=@n147.RGB, B=@n147.RGB), Alpha=@n146.G)) {name=UVs_Scale01, nodeColor=<rgba 0 0.625 1 1>}
  n148 = VectorParameter() {group=LayeredShapesUV, parameterName=LS01_Panning_Vehicle, defaultValue=<rgba 0 0 0 1>}
  n149 = VectorParameter() {group=LayeredShapesUV, parameterName=LS01_Panning_VehicleSide, defaultValue=<rgba 0 0 0 1>}
  n150 = MaterialFunctionCall(Float4=NamedRerouteUsage()) {materialFunction=&BreakOutFloat4Components}
  n151 = LinearInterpolate(
   A=LinearInterpolate(
    A=VectorParameter() {group=LayeredShapesUV, parameterName=LS01_Panning_Weapon, defaultValue=<rgba 0 0 0 1>}.RGB
    B=LinearInterpolate(A=@n148.RGB, B=@n149.RGB, Alpha=@n150.G)
    Alpha=@n150.R
   )
   B=LinearInterpolate(
    A=VectorParameter() {group=LayeredShapesUV, parameterName=LS01_Panning_Mech, defaultValue=<rgba 0 0 0 1>}.RGB
    B=VectorParameter() {group=LayeredShapesUV, parameterName=LS01_Panning_Character, defaultValue=<rgba 0 0 0 1>}.RGB
    Alpha=@n150.A
   )
   Alpha=@n150.B
  )
  n152 = VectorParameter() {group=LayeredShapesUV, parameterName=LS01_Panning_Backpack, defaultValue=<rgba 0 0 0 1>}
  n153 = MaterialFunctionCall(Float2=NamedRerouteUsage()) {materialFunction=&BreakOutFloat2Components}
  NamedRerouteDeclaration(None=LinearInterpolate(
    A=LinearInterpolate(A=@n151, B=@n152.RGB, Alpha=@n153.R)
    B=VectorParameter() {group=LayeredShapesUV, parameterName=LS01_Panning_Pickaxe, defaultValue=<rgba 0 0 0 1>}.RGB
    Alpha=@n153.G
   )) {name=UVs_Panning_01, nodeColor=<rgba 0 0.625 1 1>}
  n154 = VectorParameter() {group=LayeredShapesUV, parameterName=LS02_Panning_Vehicle, defaultValue=<rgba 0 0 0 1>}
  n155 = VectorParameter() {group=LayeredShapesUV, parameterName=LS02_Panning_VehicleSide, defaultValue=<rgba 0 0 0 1>}
  n156 = MaterialFunctionCall(Float4=NamedRerouteUsage()) {materialFunction=&BreakOutFloat4Components}
  n157 = LinearInterpolate(
   A=LinearInterpolate(
    A=VectorParameter() {group=LayeredShapesUV, parameterName=LS02_Panning_Weapon, defaultValue=<rgba 0 0 0 1>}.RGB
    B=LinearInterpolate(A=@n154.RGB, B=@n155.RGB, Alpha=@n156.G)
    Alpha=@n156.R
   )
   B=LinearInterpolate(
    A=VectorParameter() {group=LayeredShapesUV, parameterName=LS02_Panning_Mech, defaultValue=<rgba 0 0 0 1>}.RGB
    B=VectorParameter() {group=LayeredShapesUV, parameterName=LS02_Panning_Character, defaultValue=<rgba 0 0 0 1>}.RGB
    Alpha=@n156.A
   )
   Alpha=@n156.B
  )
  n158 = VectorParameter() {group=LayeredShapesUV, parameterName=LS02_Panning_Backpack, defaultValue=<rgba 0 0 0 1>}
  n159 = MaterialFunctionCall(Float2=NamedRerouteUsage()) {materialFunction=&BreakOutFloat2Components}
  NamedRerouteDeclaration(None=LinearInterpolate(
    A=LinearInterpolate(A=@n157, B=@n158.RGB, Alpha=@n159.R)
    B=VectorParameter() {group=LayeredShapesUV, parameterName=LS02_Panning_Pickaxe, defaultValue=<rgba 0 0 0 1>}.RGB
    Alpha=@n159.G
   )) {name=UVs_Panning_02, nodeColor=<rgba 0 0.625 1 1>}
 ]}
