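# Material function for the NFWrap weapon-skin wrap, applying an animated flame-style skin with separate weapon, vehicle and mech tuning selected by isVehicle and isMech.
# Bounds-relative pre-skinned positions drive top/side projections of a base texture and a panning, noise-warped flame mask.
# A gradient height-lerps that mask into a curve-atlas color over the base, whose alpha drives roughness.
# The customization mask blends the result into the incoming material behind a static switch.
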MaterialFunction /Game/Weapons/WeaponSkins/Wraps/MaterialFunctions/MF_NFWrap
graph {nodes [
  n1 = FunctionInput(Preview=VertexInterpolator(VS=PreSkinnedPosition()).PS) {bUsePreviewValueAsDefault=true, inputName=Local Position, sortPriority=11}
  n2 = PreSkinnedLocalBounds()
  n3 = VectorParameter() {group=NFWrap, parameterName=NF_Blend_Spread_Weapon_XY, sortPriority=80, defaultValue=<rgba 1 1 1 0>}
  n4 = VectorParameter() {group=NFWrap, parameterName=NF_Blend_Spread_Vehicle_XY, sortPriority=80, defaultValue=<rgba 1 1 1 0>}
  n5 = FunctionInput(Preview=Constant()) {bUsePreviewValueAsDefault=true, inputName=isMech, inputType=FunctionInput_Scalar, sortPriority=4}
  n6 = FunctionInput(Preview=Constant()) {inputName=isVehicle, inputType=FunctionInput_Scalar, sortPriority=5}
  n7 = MaterialFunctionCall!(
   ForwardAxis=FunctionInput(Preview=Constant3Vector() {constant=<rgba 1 0 0 0>}) {bUsePreviewValueAsDefault=true, inputName=ForwardAxis}
   Bounds=Divide(A=Subtract(A=@n1, B=@n2.Min), B=Subtract(A=@n2.Max, B=@n2.Max))
   Pre-Skinned Local Position
   isVehicle=@n6
   isMech=@n5
   Side Project Vehicle
   WeaponOffset=ScalarParameter() {defaultValue=-1, group=NFWrap, parameterName="NF_Blend_OffsetWeapon ", sortPriority=85}
   WeaponFalloff=ScalarParameter() {defaultValue=-0.5, group=NFWrap, parameterName="NF_Blend_FalloffWeapon (negative inverts)", sortPriority=88}
   VehicleOffset=ScalarParameter() {defaultValue=-1, group=NFWrap, parameterName="NF_Blend_OffsetVehicle ", sortPriority=86}
   VehicleFalloff=ScalarParameter() {defaultValue=-0.5, group=NFWrap, parameterName=NF_Blend_FalloffVehicle, sortPriority=89}
   MechOffset=ScalarParameter() {defaultValue=-1, group=NFWrap, parameterName="NF_Blend_OffsetMech ", sortPriority=87}
   MechFalloff=ScalarParameter() {defaultValue=-0.5, group=NFWrap, parameterName=NF_Blend_FalloffMech, sortPriority=90}
   Invert_Gradent=StaticBoolParameter() {group=NFWrap, parameterName=NF_Blend_Invert_Gradent, sortPriority=81}
   UV_Spread_XY=LinearInterpolate(A=@n3.RGB, B=@n4.RGB, Alpha=@n5)
   UV_Animated_XY=VectorParameter() {group=NFWrap, parameterName=NF_Blend_UV_Animated_XY, sortPriority=82}.RGB
   Warp_UVs
   Rotate_Mech=ScalarParameter() {defaultValue=1, group=NFWrap, parameterName=NF_Gradent_Texture_Rotate_Mech, sortPriority=74}
   Rotate_Vehicle_Side
   Rotate_Vehicle=ScalarParameter() {defaultValue=1, group=NFWrap, parameterName=NF_Gradent_Texture_Rotate_Vehicle, sortPriority=74}
  ) {materialFunction=&/MF_Wrap_Gradent_Creation}
  n8 = TextureObjectParameter() {
   group=NFWrap
   parameterName=NF_BlendGradent_Texture
   samplerType=SAMPLERTYPE_Masks
   sortPriority=74
   texture=/Game/Global/Textures/Default/Masks_and_Gradients/T_VertGradient_MASK
  }
  ChannelMaskParameter(None=TextureSample(UVs=@n7.Uvs, Tex=@n8) {samplerType=SAMPLERTYPE_Masks, texture=&T_VertGradient_MASK}.RGBA) {group=NFWrap, parameterName=NF_Blend_TextureChannel, sortPriority=74}
  n9 = TextureObjectParameter() {
   group=NFWrap
   parameterName=NF_Flame_Mask_02
   samplerType=SAMPLERTYPE_Masks
   sortPriority=62
   texture=&T_BlackMask
  }
  n10 = VectorParameter() {group=NFWrap, parameterName=NF_Fill_alternate_Axis_Color_FlameMask 02, sortPriority=67, defaultValue=<rgba 0 0 0 1>}
  n11 = VectorParameter() {group=NFWrap_UVs, parameterName=NF_Flame_Rotation-Weapon-02---------------------, sortPriority=150, defaultValue=<rgba 0 0 -0.25 1>}
  n12 = VectorParameter() {group=NFWrap_UVs, parameterName=NF_Flame_Rotation_Vehicle_02, sortPriority=151, defaultValue=<rgba 0 0 0 1>}
  n13 = VectorParameter() {group=NFWrap_UVs, parameterName=NF_Flame_Rotation_Mech_02, sortPriority=153, defaultValue=<rgba 0.25 0.25 0 1>}
  n14 = VectorParameter() {group=NFWrap_UVs, parameterName=NF_Flame_Offset_Weapon_02---------------------, sortPriority=160, defaultValue=<rgba 0 0 0 1>}
  n15 = VectorParameter() {group=NFWrap_UVs, parameterName=NF_Flame_Offset_Vehicle_02, sortPriority=161, defaultValue=<rgba 0 0 0 1>}
  n16 = LinearInterpolate(A=@n14.RGB, B=@n15.RGB, Alpha=@n6)
  n17 = VectorParameter() {group=NFWrap_UVs, parameterName=NF_Flame_Offset_Mech_02, sortPriority=162, defaultValue=<rgba 0 0 0 1>}
  n18 = MaterialFunctionCall!(
   is Vehicle=@n6
   Object Bounds Max Adjustment=FunctionInput() {bUsePreviewValueAsDefault=true, inputName=Object Bounds Max Adjustment, sortPriority=13}
   Object Bounds Min Adjustment=FunctionInput() {bUsePreviewValueAsDefault=true, inputName=Object Bounds Min Adjustment, sortPriority=12}
   Object Bounds Center Adjustment=FunctionInput() {bUsePreviewValueAsDefault=true, inputName=Object Bounds Center Adjustment, sortPriority=14}
   isMech=@n5
   Side Project Vehicle
   Local Position=@n1
  ) {materialFunction=&/MF_WrapBounds}
  n19 = Time()
  n20 = VectorParameter() {group=NFWrap_UVs, parameterName=NF_Flame_Speed_Weapon_02---------------------, sortPriority=180, defaultValue=<rgba 0 0 0 1>}
  n21 = VectorParameter() {group=NFWrap_UVs, parameterName=NF_Flame_Speed_Vehicle_02, sortPriority=181, defaultValue=<rgba 0 0 0 1>}
  n22 = VectorParameter() {group=NFWrap_UVs, parameterName=NF_Flame_Speed_Mech_02, sortPriority=182, defaultValue=<rgba 0 0 0 1>}
  n23 = LinearInterpolate(A=LinearInterpolate(A=@n20.RGB, B=@n21.RGB, Alpha=@n6), B=@n22.RGB, Alpha=@n5)
  n24 = Saturate(None=ScalarParameter() {defaultValue=1, group=NFWrap_UVs, parameterName=NF-Disable_FlameAnimations})
  n25 = Add(A=Add(A=@n18.Result, B=LinearInterpolate(A=@n16, B=@n17.RGB, Alpha=@n5)), B=Multiply(A=Multiply(A=@n19, B=@n23), B=@n24))
  n26 = VectorParameter() {group=NFWrap_UVs, parameterName="NF_Flame_Scale_Weapon_02 A=uniform", sortPriority=170, defaultValue=<rgba 1 1 1 1>}
  n27 = VectorParameter() {group=NFWrap_UVs, parameterName="NF_Flame_Scale_Vehicle_02 A=uniform", sortPriority=171, defaultValue=<rgba 1 1 1 1>}
  n28 = VectorParameter() {group=NFWrap_UVs, parameterName="NF_Flame_Scale_Mech_02 A=uniform", sortPriority=172, defaultValue=<rgba 1 1 1 1>}
  n29 = LinearInterpolate(
   A=LinearInterpolate(A=Multiply(A=@n26.RGB, B=@n26.RGB), B=Multiply(A=@n27.RGB, B=@n27.RGB), Alpha=@n6)
   B=Multiply(A=@n28.RGB, B=@n28.RGB)
   Alpha=@n5
  )
  n30 = TextureCoordinate()
  n31 = LinearInterpolate(
   A=ScalarParameter() {defaultValue=1, group=NFWrap_UVs, parameterName=NF-WarpUVS-Flame0202-TileRate Weapon, sortPriority=203}
   B=ScalarParameter() {defaultValue=1, group=NFWrap_UVs, parameterName=NF-WarpUVS-Flame0202-TileRate Vehicle, sortPriority=204}
   Alpha=@n6
  )
  n32 = VectorParameter() {group=NFWrap_UVs, parameterName=NF-WarpUVS-Flame02-Anitmated Weapon, sortPriority=207, defaultValue=<rgba 0 0 0 1>}
  n33 = VectorParameter() {group=NFWrap_UVs, parameterName=NF-WarpUVS-Flame0202-Anitmated Vehicle, sortPriority=208, defaultValue=<rgba 0 0 0 1>}
  n34 = Multiply(A=Time() {period=1}, B=ComponentMask(None=LinearInterpolate(A=@n32.RGB, B=@n33.RGB, Alpha=@n6)) {g=true, r=true})
  n35 = TextureSampleParameter2D(UVs=Add(A=Multiply(A=@n30, B=@n31), B=Multiply(A=@n34, B=@n24))) {group=NFWrap_UVs, parameterName=NF-WarpUVS-02-Texture, sortPriority=202, texture=&AtmosphericCloudNoise01}
  n36 = ChannelMaskParameter(None=@n35.RGB) {group=NFWrap_UVs, parameterName=NF-WarpUVS-Flame0202_Channel, sortPriority=201}
  n37 = LinearInterpolate(
   A=ScalarParameter() {group=NFWrap_UVs, parameterName=NF-WarpUVS-Flame0202-Amount Weapon, sortPriority=205}
   B=ScalarParameter() {group=NFWrap_UVs, parameterName=NF-WarpUVS-Flame0202-Amount_Vehicle, sortPriority=206}
   Alpha=@n6
  )
  n38 = FunctionInput(Preview=VertexInterpolator(VS=PreSkinnedNormal()).PS) {bUsePreviewValueAsDefault=true, inputName=Pre-Skinned Local Normal, sortPriority=10}
  n39 = MaterialFunctionCall!(
   IsCharacter
   IsStatsB
   Texture=@n9
   Rotation_Axis=LinearInterpolate(A=LinearInterpolate(A=@n11.RGB, B=@n12.RGB, Alpha=@n6), B=@n13.RGB, Alpha=@n5)
   Caculated_Bounds=Add(A=Multiply(A=@n25, B=@n29), B=Multiply(A=@n36.R, B=@n37))
   Pre_Skinned_Normals=@n38
   Blend_Contrast=LinearInterpolate(
    A=ScalarParameter() {defaultValue=1, group=NFWrap, parameterName=NF_Blend_Contrast, sortPriority=-1}
    B=ScalarParameter() {defaultValue=2, group=NFWrap, parameterName=NF_Blend_Contrast_Mech, sortPriority=-1}
    Alpha=@n5
   )
   Fill_alternate_Axis_Opacity=ScalarParameter() {group=NFWrap, parameterName=NF_Fill_alternate_Axis_Opacity Flame02, sortPriority=51}
   "Fill_alternate_Axis_Color_A=Roughness"=AppendVector(A=@n10.RGB, B=@n10.RGB)
   Is_Vehicle=@n6
   Is_Mech=@n5
   Enable_Fill_Z_Weapon
   Fill_Z_Weapon_opacity
   Fill_Z_Weapon_Color
  ) {materialFunction=&/MF_TopsSideProjection_Base}
  n40 = Clamp(
   None=MaterialFunctionCall(
    Transition Phase=@n7.Gradent_B
    Height Texture=ChannelMaskParameter(None=@n39.Combined) {group=NFWrap, parameterName=NF_Blend_HieghtlerpFlameMaskChannel_02, sortPriority=72}.R
    Contrast=ScalarParameter() {defaultValue=1, group=NFWrap, parameterName=NF_Blend_Hieghtlerp_Contrast, sortPriority=71}
   ) {materialFunction=&HeightLerp}.Alpha
   Min=ScalarParameter() {defaultValue=1, group=NFWrap, parameterName=NF_Blend_Hieghtlerp_Clamp_min, sortPriority=71}
   Max=ScalarParameter() {defaultValue=1, group=NFWrap, parameterName=NF_Blend_Hieghtlerp_Clamp_Max, sortPriority=71}
  )
  n41 = CurveAtlasRowParameter(CurveTime=@n40) {
   atlas=/Game/Weapons/WeaponSkins/Wraps/Textures/CruveAtlas_Wraps_01
   curve=/Game/Weapons/WeaponSkins/Wraps/Textures/C_GradSimple
   group=NFWrap
   parameterName=NF_Blend_Color_Curve
   sortPriority=71
  }
  n42 = MaterialFunctionCall!(
   Index=ScalarParameter() {group=NFWrap, parameterName="NF_Blend_Color_Curve  Index", sortPriority=71}
   CCinput=@n41
   CCinput_Alpha=@n41
   UV=@n40
  ) {materialFunction=&/MF_Global_CC_Tex}
  ComponentMask(None=@n42.3V) {b=true, g=true, r=true}
  n43 = FunctionInput() {inputName=Material, inputType=FunctionInput_MaterialAttributes}
  n44 = GetMaterialAttributes(None=@n43) {attributeGetTypes [#1, #2, #3]}
  n45 = VectorParameter() {group=NFWrap, parameterName="NF_Fill_alternate_Axis_Color_A=Roughness 01", sortPriority=3}
  n46 = VectorParameter() {group=NFWrap_UVs, parameterName=NF_Offset_Weapon_01---------------------, sortPriority=10, defaultValue=<rgba 0 0 0 1>}
  n47 = VectorParameter() {group=NFWrap_UVs, parameterName=NF_Offset_Vehicle_01, sortPriority=11, defaultValue=<rgba 0 0 0 1>}
  n48 = Add(
   A=@n18.Result
   B=LinearInterpolate(
    A=LinearInterpolate(A=@n46.RGB, B=@n47.RGB, Alpha=@n6)
    B=VectorParameter() {group=NFWrap_UVs, parameterName=NF_Offset_Mech_01, sortPriority=12, defaultValue=<rgba 0 0 0 1>}.RGB
    Alpha=@n5
   )
  )
  n49 = VectorParameter() {group=NFWrap_UVs, parameterName=NF_Panning_Weapon_01---------------------, sortPriority=40, defaultValue=<rgba 0 0 0 1>}
  n50 = VectorParameter() {group=NFWrap_UVs, parameterName=NF_Panning_Vehicle_01, sortPriority=41, defaultValue=<rgba 0 0 0 1>}
  n51 = Multiply(
   A=Time()
   B=LinearInterpolate(
    A=LinearInterpolate(A=@n49.RGB, B=@n50.RGB, Alpha=@n6)
    B=VectorParameter() {group=NFWrap_UVs, parameterName=NF_Panning_Mech_01, sortPriority=42, defaultValue=<rgba 0 0 0 1>}.RGB
    Alpha=@n5
   )
  )
  n52 = VectorParameter() {group=NFWrap_UVs, parameterName="NF_Scale_Weapon_01 A=uniform", sortPriority=30, defaultValue=<rgba 1 1 1 1>}
  n53 = VectorParameter() {group=NFWrap_UVs, parameterName="NF_Scale_Vehicle_01 A=uniform", sortPriority=31, defaultValue=<rgba 1 1 1 1>}
  n54 = LinearInterpolate(A=Multiply(A=@n52.RGB, B=@n52.RGB), B=Multiply(A=@n53.RGB, B=@n53.RGB), Alpha=@n6)
  n55 = VectorParameter() {group=NFWrap_UVs, parameterName="NF_Scale_Mech_01 A=uniform", defaultValue=<rgba 1 1 1 1>}
  n56 = Multiply(A=@n55.RGB, B=@n55.RGB)
  n57 = VectorParameter() {group=NFWrap_UVs, parameterName=NF-Rotation-Weapon-01---------------------, sortPriority=1, defaultValue=<rgba 0 0 -0.25 1>}
  n58 = VectorParameter() {group=NFWrap_UVs, parameterName=NF_Rotation_Vehicle_01, sortPriority=2, defaultValue=<rgba 0 0 0 1>}
  n59 = VectorParameter() {group=NFWrap_UVs, parameterName=NF_Rotation_Mech_01, sortPriority=3, defaultValue=<rgba 0.25 0.25 0 1>}
  n60 = MaterialFunctionCall(
   Texture=TextureObjectParameter() {group=NFWrap, parameterName=NF_BaseColor_01, sortPriority=11, texture=/Game/Weapons/WeaponSkins/Wraps/Textures/UVCheaker}
   Rotation_Axis=LinearInterpolate(A=LinearInterpolate(A=@n57.RGB, B=@n58.RGB, Alpha=@n6), B=@n59.RGB, Alpha=@n5)
   Caculated_Bounds=Multiply(A=Add(A=@n48, B=@n51), B=LinearInterpolate(A=@n54, B=@n56, Alpha=@n5))
   Pre_Skinned_Normals=@n38
   Blend_Contrast=LinearInterpolate(
    A=ScalarParameter() {defaultValue=1, group=NFWrap, parameterName=NF_Blend_Contrast, sortPriority=-1}
    B=ScalarParameter() {defaultValue=2, group=NFWrap, parameterName=NF_Blend_Contrast_Mech, sortPriority=-1}
    Alpha=@n5
   )
   Fill_alternate_Axis_Opacity=ScalarParameter() {group=NFWrap, parameterName=NF_Fill_alternate_Axis_Opacity 01, sortPriority=2}
   "Fill_alternate_Axis_Color_A=Roughness"=AppendVector(A=@n45.RGB, B=@n45.RGB)
   Is_Vehicle=@n6
   Is_Mech=@n5
  ) {materialFunction=&MF_TopsSideProjection_Base}
  n61 = SetMaterialAttributes!(
   MaterialAttributes=@n44.MaterialAttributes
   BaseColor=LinearInterpolate(A=@n60.RGB, B=@n42.3V, Alpha=@n42.3V)
   Roughness=LinearInterpolate(
    A=ScalarParameter() {group=NFWrap, parameterName=NF_Roughness_Min, sortPriority=13}
    B=ScalarParameter() {defaultValue=1, group=NFWrap, parameterName=NF_Roughness_Max, sortPriority=14}
    Alpha=@n60.Alpha
   )
   Specular=ScalarParameter() {defaultValue=0.5, group=NFWrap, parameterName=NF_Specularty, sortPriority=12}
   Metallic=ScalarParameter() {group=NFWrap, parameterName=NF_Metalic_01, sortPriority=26}
  ) {attributeSetTypes [
    #1
    #2
    #4
    #5
   ]}
  n62 = Saturate(None=Add(
    A=FunctionInput() {bUsePreviewValueAsDefault=true, inputName="Customization Mask (_CM B)", inputType=FunctionInput_Scalar, sortPriority=24}
    B=ScalarParameter() {group=NFWrap, parameterName=NF_Remove_Customization Mask, sortPriority=-10}
   ))
  FunctionOutput(None=StaticSwitchParameter(True=BlendMaterialAttributes(A=@n44.MaterialAttributes, B=@n61, Alpha=@n62), False=@n43) {group=Customization_Wrap, parameterName=NFWrap})
  LinearInterpolate(B=@n34, Alpha=@n24) {constB=0}
  LinearInterpolate(
   A=VectorParameter() {group=NFWrap_UVs, parameterName=NF-Disable_Animations Flame Shift Weapon, defaultValue=<rgba 0 0 0 1>}.RGB
   B=VectorParameter() {group=NFWrap_UVs, parameterName=NF-Disable_Animations Flame Shift Vehicle, defaultValue=<rgba 0 0 0 1>}.RGB
  )
  VectorParameter() {group=NFWrap, parameterName=NF_Blend_OffsetWeapon, sortPriority=85, defaultValue=<rgba 0 -1 0 0>}
  VectorParameter() {group=NFWrap, parameterName=NF_Blend_OffsetVehicle, sortPriority=86, defaultValue=<rgba -1 0 0 0>}
  VectorParameter() {group=NFWrap, parameterName=NF_Blend_OffsetMech, sortPriority=87, defaultValue=<rgba 0 0 -1 0>}
 ]}
