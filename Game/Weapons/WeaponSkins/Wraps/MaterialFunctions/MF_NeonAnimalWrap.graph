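# Material function for the Neon Animal weapon and vehicle wrap, painting an emissive outrun scene over black base color.
# Side-projected pre-skinned bounds coordinates, with weapon, vehicle and mech variants, sample one packed grid texture for a skewed panning grid, sun and parallax stars.
# A gradient mask splits sky from grid beneath a glowing horizon line, with time-of-day colors and Fresnel tint.
# A static switch enables blending over incoming attributes by Alpha.
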MaterialFunction /Game/Weapons/WeaponSkins/Wraps/MaterialFunctions/MF_NeonAnimalWrap
graph {nodes [
  n1 = FunctionInput(Preview=VertexInterpolator(VS=PreSkinnedPosition()).PS) {bUsePreviewValueAsDefault=true, inputName=Local Position}
  n2 = MaterialFunctionCall(Float2=Divide(A=@n1, B=FunctionInput() {inputName=Bounds_Extents})) {materialFunction=&BreakOutFloat2Components}
  n3 = FunctionInput() {inputName=isVehicle, inputType=FunctionInput_Scalar}
  n4 = LinearInterpolate(A=@n2.G, B=@n2.G, Alpha=@n3)
  Abs(None=OneMinus(None=@n4))
  n5 = VectorParameter() {desc="G=Y  B=X", group=Side Projection UVS, parameterName=SUN_Offset_Weapon, sortPriority=2, defaultValue=<rgba 0 0 0 1>}
  n6 = MaterialFunctionCall(Float2=ComponentMask(None=@n5.RGB) {g=true, r=true}) {materialFunction=&BreakOutFloat2Components}
  n7 = Frac(None=Multiply(A=Time(), B=ScalarParameter() {defaultValue=0.003, parameterName=Sunset Rate}))
  Add(A=@n6.R, B=@n7)
  n8 = VectorParameter() {desc="R=Y  G=X", group=Side Projection UVS, parameterName=SUN_Offset_Vehicle, sortPriority=4, defaultValue=<rgba 0 0 0 1>}
  n9 = MaterialFunctionCall(Float2=ComponentMask(None=@n8.RGB) {g=true, r=true}) {materialFunction=&BreakOutFloat2Components}
  Add(A=@n9.R, B=@n7)
  Add()
  n10 = FunctionInput() {inputName=PreSkin_Bounds}
  n11 = ComponentMask(None=@n10) {r=true}
  n12 = ComponentMask(None=@n10) {g=true}
  AppendVector(A=@n11, B=@n12)
  n13 = Multiply(A=@n10, B=ScalarParameter() {defaultValue=6, parameterName=Grid Tiling}) {constB=4}
  n14 = ComponentMask(None=@n13) {r=true}
  n15 = ComponentMask(None=@n13) {g=true}
  AppendVector(A=@n14, B=@n15)
  Constant3Vector() {constant=<rgba 0.9271 0.0628 0.0628 0>}
  n16 = FunctionInput() {inputType=FunctionInput_MaterialAttributes}
  n17 = GetMaterialAttributes(None=@n16) {attributeGetTypes [#1, #2, #3]}
  n18 = VectorParameter() {parameterName=BG_Tint 4, defaultValue=<rgba 0 0 0 1>}
  n19 = VectorParameter() {parameterName=BG_Tint, defaultValue=<rgba 0.3553 0.1698 0.875 1>}
  n20 = VectorParameter() {parameterName=BG_Tint 2, defaultValue=<rgba 0.4 0.0497 0 1>}
  n21 = Constant()
  n22 = VectorParameter() {parameterName=BG_Tint 3, defaultValue=<rgba 0.1165 0 0.5 1>}
  n23 = ScalarParameter() {defaultValue=0.05, parameterName=BG_Fresnel}
  n24 = ScalarParameter() {defaultValue=1, parameterName=BG_Fresnel2}
  n25 = Fresnel(ExponentIn=MaterialFunctionCall!(Color_Morning=@n23, Color_Day=@n23, Color_Evening=@n23, Color_Night=@n24) {materialFunction=&/Wrap_Color_Time_of_Day}.Result) {baseReflectFraction=0}
  n26 = LinearInterpolate(
   A=MaterialFunctionCall(Color_Morning=@n18.RGB, Color_Day=@n19.RGB, Color_Evening=@n20.RGB, Color_Night=@n21) {materialFunction=&Wrap_Color_Time_of_Day}.Result
   B=MaterialFunctionCall(Color_Morning=@n21, Color_Day=@n21, Color_Evening=@n21, Color_Night=@n22.RGB) {materialFunction=&Wrap_Color_Time_of_Day}.Result
   Alpha=@n25
  ) {constB=0}
  n27 = FunctionInput() {inputName=isMech, inputType=FunctionInput_Scalar}
  n28 = FunctionInput() {bUsePreviewValueAsDefault=true, inputName=Side Project Vehicle, inputType=FunctionInput_Scalar}
  n29 = ComponentMask(None=MaterialFunctionCall!(isVehicle=@n3, isMech=@n27, Side Project Vehicle=@n28, In=@n10, IsBackpack, IsPickaxe, isCharacter) {materialFunction=&/MF_SideProjection_Math}.Result) {r=true}
  n30 = Saturate(None=LinearInterpolate(A=ScalarParameter() {defaultValue=-1, parameterName=LEXR}, B=ScalarParameter() {defaultValue=1.5, parameterName=LEXR3}, Alpha=@n29))
  n31 = LinearInterpolate(
   A=VectorParameter() {parameterName=GridColor_C, defaultValue=<rgba 1.0927 0 4 1>}.RGB
   B=VectorParameter() {parameterName=GridColor_D, defaultValue=<rgba 1.5 0.9159 0 1>}.RGB
   Alpha=@n30
  )
  n32 = LinearInterpolate(
   A=VectorParameter() {parameterName=GridColor_B, defaultValue=<rgba 0.2919 0 4 1>}.RGB
   B=VectorParameter() {parameterName=GridColor_A, defaultValue=<rgba 2 0 0.5825 1>}.RGB
   Alpha=@n30
  )
  n33 = LinearInterpolate(
   A=VectorParameter() {parameterName=GridColor_E, defaultValue=<rgba 1.819 4 3.7607 1>}.RGB
   B=VectorParameter() {parameterName=GridColor_F, defaultValue=<rgba 0 0.1122 2 1>}.RGB
   Alpha=@n30
  )
  n34 = MaterialFunctionCall(Color_Morning=@n31, Color_Day=@n32, Color_Evening=@n31, Color_Night=@n33) {materialFunction=&Wrap_Color_Time_of_Day}
  n35 = Divide(A=@n4)
  n36 = ScalarParameter() {defaultValue=0.4, group=Side Projection, parameterName=Side Projection Skew Factor}
  n37 = LinearInterpolate(A=@n36, B=Multiply(A=@n36, B=ScalarParameter() {defaultValue=12, parameterName=SkewMult}) {constB=1.5}, Alpha=@n29)
  n38 = VectorParameter() {desc="G=X  B=Y A=Uniform", group=Side Projection UVS, parameterName=UVs_Scale_Weapon, sortPriority=1, defaultValue=<rgba 1 1 1 1>}
  n39 = VectorParameter() {desc="G=Y  B=X", group=Side Projection UVS, parameterName=UVs_Offset_Weapon, sortPriority=2, defaultValue=<rgba 0 0 0 1>}
  n40 = MaterialFunctionCall(Float2=ComponentMask(None=@n39.RGB) {g=true, r=true}) {materialFunction=&BreakOutFloat2Components}
  n41 = Frac(None=Multiply(A=Time(), B=ScalarParameter() {defaultValue=0.25, parameterName=PanningSpeed}))
  n42 = Add(A=@n40.R, B=@n41)
  n43 = ComponentMask(None=@n13) {b=true}
  n44 = Multiply(
   A=AppendVector(A=@n38.R, B=@n38.R)
   B=Add(
    A=MaterialFunctionCall(X=@n42, Y=@n40.G) {materialFunction=&MakeFloat2}.Result
    B=AppendVector(A=@n15, B=@n43)
   )
  )
  n45 = StaticSwitchParameter(True=AppendVector(A=@n15, B=@n14), False=LinearInterpolate(A=AppendVector(A=@n14, B=@n15), B=AppendVector(A=@n14, B=@n43), Alpha=@n28)) {group=Side Projection UVS, parameterName=Rotate Vehicle Projection, sortPriority=-4}
  n46 = VectorParameter() {desc="R=Y  G=X", group=Side Projection UVS, parameterName=UVs_Offset_Vehicle, sortPriority=4, defaultValue=<rgba 0 0 0 1>}
  n47 = MaterialFunctionCall(Float2=ComponentMask(None=@n46.RGB) {g=true, r=true}) {materialFunction=&BreakOutFloat2Components}
  n48 = MaterialFunctionCall(X=Add(A=@n47.R, B=@n41), Y=@n47.G) {materialFunction=&MakeFloat2}
  n49 = VectorParameter() {desc="R=  G=  A=Uniform", group=Side Projection UVS, parameterName=UVs_Scale_Vehicle, sortPriority=3, defaultValue=<rgba 1 1 1 1>}
  n50 = AppendVector(A=@n43, B=@n14)
  n51 = VectorParameter() {desc="R=Y  G=X", group=Side Projection UVS, parameterName=UVs_Offset_Mech, sortPriority=6, defaultValue=<rgba 0 0 0 1>}
  n52 = MaterialFunctionCall(Float2=ComponentMask(None=@n51.RGB) {g=true, r=true}) {materialFunction=&BreakOutFloat2Components}
  n53 = MaterialFunctionCall(X=Add(A=@n52.R, B=@n41), Y=@n52.G) {materialFunction=&MakeFloat2}
  n54 = VectorParameter() {desc="R=  G=  A=Uniform", group=Side Projection UVS, parameterName=UVs_Scale_Mech, sortPriority=5, defaultValue=<rgba 1 1 1 1>}
  n55 = LinearInterpolate(
   A=Multiply(A=Add(A=@n45, B=@n48.Result), B=AppendVector(A=@n49.R, B=@n49.R))
   B=Multiply(A=Add(A=@n50, B=@n53.Result), B=AppendVector(A=@n54.R, B=@n54.R))
   Alpha=@n27
  )
  n56 = TextureSample(UVs=Add(A=Multiply(A=@n35, B=@n37), B=LinearInterpolate(A=@n44, B=@n55, Alpha=@n3))) {samplerType=SAMPLERTYPE_LinearColor, texture=/Game/Weapons/WeaponSkins/Wraps/Textures/NeonAnimal/OutrunWrap_Grid_M}
  n57 = FunctionInput(Preview=VertexInterpolator(VS=PreSkinnedNormal()).PS) {bUsePreviewValueAsDefault=true, inputName=Local normal}
  n58 = Abs(None=ComponentMask(None=@n57) {g=true})
  n59 = Abs(None=ComponentMask(None=@n57) {r=true})
  n60 = Saturate(None=LinearInterpolate(A=Power(Base=@n58), B=Power(Base=@n59), Alpha=@n3))
  n61 = Saturate(None=Subtract(A=@n56.B, B=@n60))
  n62 = VectorParameter() {parameterName=Star Color, defaultValue=<rgba 0.75 0.1472 0.3371 1>}
  n63 = VectorParameter() {parameterName=Star Color 2, defaultValue=<rgba 0 0 0 1>}
  n64 = VectorParameter() {desc="G=X  B=Y A=Uniform", group=Side Projection UVS, parameterName=UVs_Scale_Weapon, sortPriority=1, defaultValue=<rgba 1 1 1 1>}
  n65 = AppendVector(A=@n64.R, B=@n64.R)
  n66 = ScalarParameter() {defaultValue=12, parameterName=Star Scale}
  n67 = VectorParameter() {desc="G=Y  B=X", group=Side Projection UVS, parameterName=UVs_Offset_Weapon, sortPriority=2, defaultValue=<rgba 0 0 0 1>}
  n68 = ComponentMask(None=@n10) {b=true}
  n69 = Add(A=ComponentMask(None=@n67.RGB) {g=true, r=true}, B=AppendVector(A=@n12, B=@n68))
  n70 = VectorParameter() {desc="G=Y  B=X", group=Side Projection UVS, parameterName=SUN_Offset_Weapon, sortPriority=2, defaultValue=<rgba 0 0 0 1>}
  n71 = ComponentMask(None=@n70.RGB) {g=true, r=true}
  n72 = AppendVector(A=@n12, B=@n11)
  n73 = LinearInterpolate(A=AppendVector(A=@n11, B=@n12), B=AppendVector(A=@n11, B=@n68), Alpha=@n28)
  n74 = VectorParameter() {desc="R=Y  G=X", group=Side Projection UVS, parameterName=UVs_Offset_Vehicle, sortPriority=4, defaultValue=<rgba 0 0 0 1>}
  n75 = VectorParameter() {desc="R=Y  G=X", group=Side Projection UVS, parameterName=SUN_Offset_Vehicle, sortPriority=4, defaultValue=<rgba 0 0 0 1>}
  n76 = Add(
   A=Add(
    A=StaticSwitchParameter(True=@n72, False=@n73) {group=Side Projection UVS, parameterName=Rotate Vehicle Projection, sortPriority=-4}
    B=ComponentMask(None=@n74.RGB) {g=true, r=true}
   )
   B=ComponentMask(None=@n75.RGB) {g=true, r=true}
  )
  n77 = VectorParameter() {desc="R=  G=  A=Uniform", group=Side Projection UVS, parameterName=UVs_Scale_Vehicle, sortPriority=3, defaultValue=<rgba 1 1 1 1>}
  n78 = Multiply(A=AppendVector(A=@n77.R, B=@n77.R), B=@n66)
  n79 = VectorParameter() {desc="R=Y  G=X", group=Side Projection UVS, parameterName=UVs_Offset_Mech, sortPriority=6, defaultValue=<rgba 0 0 0 1>}
  n80 = VectorParameter() {desc="R=  G=  A=Uniform", group=Side Projection UVS, parameterName=UVs_Scale_Mech, sortPriority=5, defaultValue=<rgba 1 1 1 1>}
  n81 = Multiply(A=Add(A=AppendVector(A=@n68, B=@n11), B=ComponentMask(None=@n79.RGB) {g=true, r=true}), B=AppendVector(A=@n80.R, B=@n80.R))
  n82 = LinearInterpolate(A=Multiply(A=Multiply(A=@n65, B=@n66), B=Add(A=@n69, B=@n71)), B=LinearInterpolate(A=Multiply(A=@n76, B=@n78), B=Multiply(A=@n81, B=@n66)))
  n83 = ScalarParameter() {defaultValue=-3, parameterName=Star Height}
  n84 = Multiply(
   A=MaterialFunctionCall(Color_Morning=@n62.R, Color_Day=@n63.RGB, Color_Evening=@n62.R, Color_Night=@n62.R) {materialFunction=&Wrap_Color_Time_of_Day}.Result
   B=TextureSample(UVs=BumpOffset(Coordinate=@n82, Height=@n83)) {samplerType=SAMPLERTYPE_LinearColor, texture=&OutrunWrap_Grid_M}.R
  )
  n85 = Saturate(None=LinearInterpolate(A=ScalarParameter() {defaultValue=-2.5, parameterName=LEXR2}, B=ScalarParameter() {defaultValue=1.2, parameterName=8456}, Alpha=@n29))
  n86 = LinearInterpolate(
   A=VectorParameter() {parameterName=Sun Color_B, defaultValue=<rgba 2 0 0.2057 1>}.RGB
   B=VectorParameter() {parameterName=Sun Color_A, defaultValue=<rgba 7 2.9869 0 1>}.RGB
   Alpha=@n85
  )
  n87 = LinearInterpolate(
   A=VectorParameter() {parameterName=Sun Color_C, defaultValue=<rgba 0.6963 0 2 1>}.RGB
   B=VectorParameter() {parameterName=Sun Color_D, defaultValue=<rgba 7 0 5.72 1>}.RGB
   Alpha=@n85
  )
  n88 = VectorParameter() {desc="G=X  B=Y A=Uniform", group=Side Projection UVS, parameterName=UVs_Scale_Weapon, sortPriority=1, defaultValue=<rgba 1 1 1 1>}
  n89 = VectorParameter() {desc="G=Y  B=X", group=Side Projection UVS, parameterName=UVs_Offset_Weapon, sortPriority=2, defaultValue=<rgba 0 0 0 1>}
  n90 = ComponentMask(None=@n10) {g=true}
  n91 = ComponentMask(None=@n10) {b=true}
  n92 = Add(A=ComponentMask(None=@n89.RGB) {g=true, r=true}, B=AppendVector(A=@n90, B=@n91))
  n93 = MaterialFunctionCall(X=@n6.R, Y=@n6.R) {materialFunction=&MakeFloat2}
  n94 = ComponentMask(None=@n10) {r=true}
  n95 = AppendVector(A=@n90, B=@n94)
  n96 = LinearInterpolate(A=AppendVector(A=@n94, B=@n90), B=AppendVector(A=@n94, B=@n91), Alpha=@n28)
  n97 = VectorParameter() {desc="R=Y  G=X", group=Side Projection UVS, parameterName=UVs_Offset_Vehicle, sortPriority=4, defaultValue=<rgba 0 0 0 1>}
  n98 = Add(
   A=Add(
    A=StaticSwitchParameter(True=@n95, False=@n96) {group=Side Projection UVS, parameterName=Rotate Vehicle Projection, sortPriority=-4}
    B=ComponentMask(None=@n97.RGB) {g=true, r=true}
   )
   B=MaterialFunctionCall(X=@n9.R, Y=@n9.R) {materialFunction=&MakeFloat2}.Result
  )
  n99 = VectorParameter() {desc="R=  G=  A=Uniform", group=Side Projection UVS, parameterName=UVs_Scale_Vehicle, sortPriority=3, defaultValue=<rgba 1 1 1 1>}
  n100 = AppendVector(A=@n99.R, B=@n99.R)
  n101 = VectorParameter() {desc="R=Y  G=X", group=Side Projection UVS, parameterName=UVs_Offset_Mech, sortPriority=6, defaultValue=<rgba 0 0 0 1>}
  n102 = Add(A=AppendVector(A=@n91, B=@n94), B=ComponentMask(None=@n101.RGB) {g=true, r=true})
  n103 = VectorParameter() {desc="R=  G=  A=Uniform", group=Side Projection UVS, parameterName=UVs_Scale_Mech, sortPriority=5, defaultValue=<rgba 1 1 1 1>}
  n104 = AppendVector(A=@n103.R, B=@n103.R)
  n105 = LinearInterpolate(
   A=Multiply(A=AppendVector(A=@n88.R, B=@n88.R), B=Add(A=@n92, B=@n93.Result))
   B=LinearInterpolate(A=Multiply(A=@n98, B=@n100), B=Multiply(A=@n102, B=@n104), Alpha=@n27)
   Alpha=@n3
  )
  n106 = Saturate(None=Subtract(
    A=TextureSample(UVs=@n105) {samplerType=SAMPLERTYPE_LinearColor, texture=&OutrunWrap_Grid_M}.G
    B=@n60
   ))
  n107 = LinearInterpolate(
   A=Multiply(A=@n26, B=Constant3Vector() {constant=<rgba 0.1458 0.0076 0.0351 0>}) {constB=0.2}
   B=MaterialFunctionCall(Color_Morning=@n86, Color_Day=@n86, Color_Evening=@n86, Color_Night=@n87) {materialFunction=&Wrap_Color_Time_of_Day}.Result
   Alpha=@n106
  )
  n108 = Constant() {r=1}
  n109 = ScalarParameter() {defaultValue=0.38, parameterName=Rad}
  n110 = Constant() {r=1}
  n111 = VectorParameter() {parameterName=Horizon Line Color, defaultValue=<rgba 0 2.6737 7 1>}
  n112 = SphereMask(
   A=@n29
   B=ScalarParameter() {defaultValue=0.62, parameterName=BHoriz}
   Radius=ScalarParameter() {defaultValue=0.0025, parameterName=Rad2}
   Hardness=ScalarParameter() {defaultValue=0.05, parameterName=Hardness}
  )
  n113 = LinearInterpolate(
   A=LinearInterpolate(
    A=LinearInterpolate(A=@n26, B=@n34.Result, Alpha=@n61)
    B=LinearInterpolate(A=@n84, B=@n107, Alpha=@n106)
    Alpha=SphereMask(A=@n29, B=@n108, Radius=@n109, Hardness=@n110)
   ) {constB=0}
   B=@n111.RGB
   Alpha=@n112
  )
  n114 = Saturate(None=Power(Base=@n25) {constExponent=58})
  n115 = SetMaterialAttributes!(
   MaterialAttributes=@n17.MaterialAttributes
   EmissiveColor=Add(A=@n17.EmissiveColor, B=LinearInterpolate(A=@n113, Alpha=@n114) {constB=0})
   Roughness=Saturate(None=Multiply(A=@n17.Roughness) {constB=0.6})
   BaseColor=Constant()
   Opacity=Constant()
  ) {attributeSetTypes [
    #1
    #2
    #4
    #5
   ]}
  n116 = FunctionInput() {inputName=Alpha, inputType=FunctionInput_Scalar}
  FunctionOutput(None=StaticSwitchParameter(True=BlendMaterialAttributes(A=@n16, B=@n115, Alpha=@n116), False=@n16) {group=Customization_Wrap, parameterName=Use Neon Animal Wrap})
  MaterialFunctionCall(Float2=@n35) {materialFunction=&BreakOutFloat2Components}
  MaterialFunctionCall!(
   is Vehicle=@n3
   Object Bounds Max Adjustment=FunctionInput() {bUsePreviewValueAsDefault=true, inputName=Object Bounds Max Adjustment}
   Object Bounds Min Adjustment=FunctionInput() {bUsePreviewValueAsDefault=true, inputName=Object Bounds Min Adjustment}
   Object Bounds Center Adjustment=FunctionInput() {bUsePreviewValueAsDefault=true, inputName=Object Bounds Center Adjustment}
   isMech=@n27
   Side Project Vehicle=@n28
   Local Position=@n1
  ) {materialFunction=&/MF_WrapBounds}
  MaterialFunctionCall(Color_Morning=@n31, Color_Day=@n32, Color_Evening=@n31, Color_Night=@n33) {materialFunction=&Color_Time_of_Day}
  MaterialFunctionCall(Color_Morning=@n86, Color_Day=@n86, Color_Evening=@n86, Color_Night=@n87) {materialFunction=&Color_Time_of_Day}
  MaterialFunctionCall(Color_Morning=@n62.R, Color_Day=@n63.RGB, Color_Evening=@n62.R, Color_Night=@n62.R) {materialFunction=&Color_Time_of_Day}
  MaterialFunctionCall(Color_Morning=@n18.RGB, Color_Day=@n19.RGB, Color_Evening=@n20.RGB, Color_Night=@n21) {materialFunction=&Color_Time_of_Day}
  MaterialFunctionCall(Color_Morning=@n21, Color_Day=@n21, Color_Evening=@n21, Color_Night=@n22.RGB) {materialFunction=&Color_Time_of_Day}
  MaterialFunctionCall(Value_Morning=@n23, Value_Day=@n23, Value_Evening=@n23, Value_Night=@n24) {materialFunction=&Scalar_Time_of_Day}
  Max(A=@n84, B=@n107)
  Multiply(A=@n111.RGB, B=@n112)
  Multiply(A=@n56.B, B=@n34.Result)
  Transform(None=@n17.Normal)
 ]}
