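# Weapon-wrap material function that adds an animated neon sign look to weapons, vehicles, mechs and characters, behind the Wrap_NeonSign static switch.
# Top/side projection from pre-skinned bounds maps RGB masks to red, yellow and blue tubes.
# Offset sine timings and scrolling noise flicker each channel's glow, while a Blueprint-fed game time triggers an all-channel flash on fire that decays.
# Customization masks choose which regions blend over the incoming material attributes.
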MaterialFunction /Game/Weapons/WeaponSkins/Wraps/MaterialFunctions/MF_NeonSign_Wrap
graph {nodes [
  n1 = FunctionInput() {inputName=Material, inputType=FunctionInput_MaterialAttributes}
  n2 = GetMaterialAttributes(None=@n1) {attributeGetTypes [#1, #2, #3]}
  n3 = TextureObjectParameter() {
   group=NeonSign
   parameterName=NS_Normal_01
   samplerType=SAMPLERTYPE_Normal
   sortPriority=13
   texture=&FlatNormalBlue
  }
  n4 = Constant()
  n5 = FunctionInput(Preview=@n4) {bUsePreviewValueAsDefault=true, inputName=isMech, inputType=FunctionInput_Scalar, sortPriority=4}
  n6 = LinearInterpolate(
   A=ScalarParameter() {defaultValue=1, group=NeonSign, parameterName=NS_Fill_alternate_Axis_BlendContrast, sortPriority=19}
   B=ScalarParameter() {defaultValue=2, group=NeonSign, parameterName=NS_Fill_alternate_Axis_BlendContrast_Mech, sortPriority=19}
   Alpha=@n5
  )
  n7 = ScalarParameter() {defaultValue=1, group=NeonSign, parameterName=NS_Fill_alternate_Axis_Opacity, sortPriority=19}
  n8 = VectorParameter() {group=NeonSign_UVs, parameterName=NS_WeaponRotation, sortPriority=-4, defaultValue=<rgba 0 0 -0.25 1>}
  n9 = VectorParameter() {group=NeonSign_UVs, parameterName=NS_VehicleRotation, sortPriority=-3, defaultValue=<rgba 0 0 0 1>}
  n10 = FunctionInput(Preview=Constant()) {inputName=isVehicle, inputType=FunctionInput_Scalar, sortPriority=5}
  n11 = VectorParameter() {group=NeonSign_UVs, parameterName=NS_MechRotation, sortPriority=-2, defaultValue=<rgba 0.2502 0.2502 0 1>}
  n12 = VectorParameter() {group=NeonSign_UVs, parameterName=NS_CharacterRotation, sortPriority=-2, defaultValue=<rgba 0.2502 0.2502 0 1>}
  n13 = FunctionInput(Preview=@n4) {bUsePreviewValueAsDefault=true, inputName=isCharacter, inputType=FunctionInput_Scalar, sortPriority=4}
  n14 = LinearInterpolate(A=LinearInterpolate(A=@n8.RGB, B=@n9.RGB, Alpha=@n10), B=LinearInterpolate(A=@n11.RGB, B=@n12.RGB, Alpha=@n13), Alpha=@n5)
  n15 = FunctionInput(Preview=VertexInterpolator(VS=PreSkinnedNormal()).PS) {bUsePreviewValueAsDefault=true, inputName=Pre-Skinned Local Normal, sortPriority=10}
  n16 = LinearInterpolate(
   A=VectorParameter() {group=NeonSign_UVs, parameterName=NS_Offset_Weapon, sortPriority=-4, defaultValue=<rgba 0 0 0 1>}.RGB
   B=VectorParameter() {group=NeonSign_UVs, parameterName=NS_Offset_Vehicle, sortPriority=-3, defaultValue=<rgba 0 0 0 1>}.RGB
   Alpha=@n10
  )
  n17 = VectorParameter() {group=NeonSign_UVs, parameterName=NS_Offset_Character, sortPriority=-2, defaultValue=<rgba 0 0 0 1>}
  n18 = LinearInterpolate(
   A=VectorParameter() {group=NeonSign_UVs, parameterName=NS_Offset_Mech, sortPriority=-2, defaultValue=<rgba 0 0 0 1>}.RGB
   B=@n17.RGB
   Alpha=@n13
  )
  n19 = PreSkinnedPosition()
  n20 = PreSkinnedLocalBounds()
  n21 = Divide(A=Subtract(A=VertexInterpolator(VS=@n19).PS, B=@n20.Min), B=Subtract(A=@n20.Max, B=@n20.Max))
  n22 = VectorParameter() {group=NeonSign_UVs, parameterName=NS_Scale_Weapon, sortPriority=-4, defaultValue=<rgba 1 1 1 1>}
  n23 = VectorParameter() {group=NeonSign_UVs, parameterName=NS_Scale_Vehicle, sortPriority=-3, defaultValue=<rgba 1 1 1 1>}
  n24 = VectorParameter() {group=NeonSign_UVs, parameterName=NS_Scale_Mech, sortPriority=-2, defaultValue=<rgba 1 1 1 1>}
  n25 = VectorParameter() {group=NeonSign_UVs, parameterName=NS_Scale_Character, sortPriority=-2, defaultValue=<rgba 1 1 1 1>}
  n26 = Multiply(
   A=Add(
    A=FunctionInput(Preview=@n21) {bUsePreviewValueAsDefault=true, inputName=Calcated_Bounds, sortPriority=11}
    B=LinearInterpolate(A=@n16, B=@n18, Alpha=@n5)
   )
   B=LinearInterpolate(A=LinearInterpolate(A=@n22.RGB, B=@n23.RGB, Alpha=@n10), B=LinearInterpolate(A=@n24.RGB, B=@n25.RGB, Alpha=@n13), Alpha=@n5)
  )
  n27 = LinearInterpolate(
   A=@n2.Normal
   B=Add(
    A=@n2.Normal
    B=MaterialFunctionCall!(
     Texture_Normals=@n3
     Rotation_Axis=@n14
     Caculated_Bounds=@n26
     Pre_Skinned_Normals=@n15
     Blend_Contrast=@n6
     Fill_alternate_Axis_Opacity=@n7
     Is_Vehicle=@n10
     Is_Mech=@n5
    ) {materialFunction=&/MF_TopsSideProjection_Normals}.Add_Surface_Normal
   )
   Alpha=ScalarParameter() {defaultValue=1, group=NeonSign, parameterName=NS_NormalIntensity, sortPriority=4}
  )
  n28 = ShadingPathSwitch(Default=@n27, Deferred=@n27, Mobile=@n2.Normal) {desc=Migrated from FeatureLevelSwitch}
  n29 = VectorParameter() {group=NeonSign, parameterName=NS_BlueColor, sortPriority=1, defaultValue=<rgba 0.0222 0.9387 0.9911 1>}
  n30 = VectorParameter() {group=NeonSign, parameterName="NS_Fill_alternate_Axis_Color_A=Roughness 01", sortPriority=19}
  n31 = AppendVector(A=@n30.RGB, B=@n30.RGB)
  n32 = MaterialFunctionCall!(
   IsCharacter
   IsStatsB
   Texture=TextureObjectParameter() {group=NeonSign, parameterName=NS_Mask_01, sortPriority=10, texture=&T_BlacksRGB}
   Rotation_Axis=@n14
   Caculated_Bounds=@n26
   Pre_Skinned_Normals=@n15
   Blend_Contrast=@n6
   Fill_alternate_Axis_Opacity=@n7
   "Fill_alternate_Axis_Color_A=Roughness"=@n31
   Is_Vehicle=@n10
   Is_Mech=@n5
   Enable_Fill_Z_Weapon
   Fill_Z_Weapon_opacity
   Fill_Z_Weapon_Color
  ) {materialFunction=&/MF_TopsSideProjection_Base}
  n33 = ComponentMask(None=@n32.Combined) {b=true}
  n34 = MaterialFunctionCall(
   Texture=TextureObjectParameter() {group=NeonSign, parameterName=NS_Mask_02, sortPriority=11, texture=&T_BlacksRGB}
   Rotation_Axis=@n14
   Caculated_Bounds=@n26
   Pre_Skinned_Normals=@n15
   Blend_Contrast=@n6
   Fill_alternate_Axis_Opacity=@n7
   "Fill_alternate_Axis_Color_A=Roughness"=@n31
   Is_Vehicle=@n10
   Is_Mech=@n5
  ) {materialFunction=&MF_TopsSideProjection_Base}
  n35 = ScalarParameter() {defaultValue=0.5, group=NeonSign, parameterName=NS_Fluorescent_GlowIntensity, sortPriority=58}
  n36 = Add(A=@n33, B=Multiply(A=ComponentMask(None=@n34.Combined) {b=true}, B=@n35))
  n37 = TextureCoordinate()
  n38 = ScalarParameter() {defaultValue=1, group=NeonSign, parameterName=NS_EmissiveVariation_TileRate, sortPriority=52}
  n39 = Time() {period=1}
  n40 = VectorParameter() {group=NeonSign, parameterName=NS_EmissiveVariation_Animation, sortPriority=53, defaultValue=<rgba 0.05 0.1 5 1>}
  n41 = ComponentMask(None=@n40.RGB) {g=true, r=true}
  n42 = TextureSampleParameter2D(UVs=Add(A=Multiply(A=@n37, B=@n38), B=Multiply(A=@n39, B=@n41))) {group=NeonSign, parameterName=NS_FluorescentTexture, sortPriority=52, texture=&T_BlacksRGB}
  n43 = ChannelMaskParameter(None=@n42.RGBA) {group=NeonSign, maskChannel=Blue, parameterName=NS_FluorescentMaskChannel, sortPriority=52, defaultValue=<rgba 0 0 1 0>}
  n44 = Multiply(A=ScalarParameter() {defaultValue=0.025, group=NeonSign, parameterName=NS_Fluorescent_NoiseIntensity, sortPriority=54}, B=@n43.R)
  n45 = ScalarParameter() {defaultValue=0.025, group=NeonSign, parameterName=NS_AmbientEmissiveIntensity, sortPriority=53}
  n46 = Multiply(A=ScalarParameter() {defaultValue=0.5, group=NeonSign, parameterName=NS_TimingVariation, sortPriority=49}, B=Time())
  n47 = Sine(None=@n46) {period=2.1}
  n48 = Add(A=@n44, B=Add(A=@n45, B=Saturate(None=@n47)))
  n49 = VectorParameter() {group=NeonSign, parameterName=NS_YellowColor, sortPriority=1, defaultValue=<rgba 0.8796 0.8796 0 1>}
  n50 = ComponentMask(None=@n32.Combined) {g=true}
  n51 = Add(A=@n50, B=Multiply(A=ComponentMask(None=@n34.Combined) {g=true}, B=@n35))
  n52 = AppendVector(A=Sine(None=@n46) {period=1.6}, B=Sine(None=@n46) {period=1.42})
  n53 = Desaturation(None=@n52)
  n54 = Add(A=@n44, B=Add(A=@n45, B=Saturate(None=@n53)))
  n55 = Add(
   A=Multiply(A=@n29.RGB, B=LinearInterpolate(B=@n36, Alpha=@n48) {desc=blue})
   B=Multiply(A=@n49.RGB, B=LinearInterpolate(B=@n51, Alpha=@n54) {desc=yellow})
  )
  n56 = VectorParameter() {group=NeonSign, parameterName=NS_RedColor, sortPriority=1, defaultValue=<rgba 0.9911 0.0513 0.0012 1>}
  n57 = ComponentMask(None=@n32.Combined) {r=true}
  n58 = Multiply(A=ComponentMask(None=@n34.Combined) {r=true}, B=@n35)
  n59 = Ceil(None=AppendVector(A=@n52, B=Sine(None=@n46) {period=1.51}))
  n60 = Add(A=@n45, B=Saturate(None=Desaturation(None=@n59)))
  n61 = Multiply(A=@n56.RGB, B=LinearInterpolate(B=Add(A=@n57, B=@n58), Alpha=Add(A=@n44, B=@n60)) {desc=red})
  n62 = Multiply(A=@n56.RGB, B=@n57)
  n63 = Multiply(A=@n49.RGB, B=@n50)
  n64 = Multiply(
   A=Add(A=Add(A=@n62, B=@n63), B=Multiply(A=@n29.RGB, B=@n33))
   B=ScalarParameter() {defaultValue=0.5, group=NeonSign_Reactive, parameterName=NS_EmissiveColorMultiplier, sortPriority=0}
  )
  n65 = MaterialFunctionCall(
   Texture=TextureObjectParameter() {group=NeonSign, parameterName=NS_Emissive_01, sortPriority=12, texture=&T_BlacksRGB}
   Rotation_Axis=@n14
   Caculated_Bounds=@n26
   Pre_Skinned_Normals=@n15
   Blend_Contrast=@n6
   Fill_alternate_Axis_Opacity=@n7
   "Fill_alternate_Axis_Color_A=Roughness"=@n31
   Is_Vehicle=@n10
   Is_Mech=@n5
  ) {materialFunction=&MF_TopsSideProjection_Base}
  n66 = LinearInterpolate(
   A=@n64
   B=Multiply(
    A=ComponentMask(None=@n65.Combined) {b=true, g=true, r=true}
    B=ScalarParameter() {defaultValue=1, group=NeonSign_Reactive, parameterName=NS_EmissiveTextureMultiplier, sortPriority=1}
   )
   Alpha=ScalarParameter() {defaultValue=1, group=NeonSign_Reactive, parameterName=NS_EmissiveTextureBlendAmount, sortPriority=2}
  )
  n67 = Divide(
   A=Subtract(A=Time(), B=ScalarParameter() {group=00-Reactive-GTFB, parameterName=00-GameTimeFromBP})
   B=ScalarParameter() {defaultValue=1, group=NeonSign_Reactive, parameterName=NS_AnimationDuration, sortPriority=33}
  )
  n68 = OneMinus(None=Power(Base=Saturate(None=@n67), Exp=ScalarParameter() {defaultValue=0.5, group=NeonSign_Reactive, parameterName=NS_Curve_Amount, sortPriority=34})) {desc=inverts curve from 0-1 to 1-0}
  n69 = ScalarParameter() {group=NeonSign_Reactive, parameterName=NS_TestReactive, sortPriority=300}
  n70 = ScalarParameter() {defaultValue=0.5, group=NeonSign, parameterName=NS_Specular_Min, sortPriority=5}
  n71 = ScalarParameter() {defaultValue=1, group=NeonSign, parameterName=NS_Specular_Max, sortPriority=6}
  n72 = ComponentMask(None=@n34.Combined) {a=true}
  n73 = ScalarParameter() {group=NeonSign, parameterName=NS_Roughness_Min, sortPriority=8}
  n74 = ScalarParameter() {defaultValue=1, group=NeonSign, parameterName=NS_Roughness_Max, sortPriority=9}
  n75 = ComponentMask(None=@n32.Combined) {a=true}
  n76 = VectorParameter() {group=NeonSign, parameterName=NS_BaseColor, sortPriority=0, defaultValue=<rgba 0.0037 0.0052 0.0273 1>}
  n77 = StaticSwitchParameter(True=@n1, False=@n76.RGB) {defaultValue=true, group=NeonSign, parameterName=Use other Function as Base, sortPriority=-1}
  n78 = ScalarParameter() {defaultValue=0.025, group=NeonSign, parameterName=NS_BaseColorBlend, sortPriority=2}
  n79 = LinearInterpolate(A=@n77, B=Multiply(A=@n57, B=LinearInterpolate(A=@n77, B=@n56.RGB, Alpha=@n78)), Alpha=@n57)
  n80 = Multiply(A=@n50, B=LinearInterpolate(A=@n77, B=@n49.RGB, Alpha=@n78))
  n81 = LinearInterpolate(A=@n77, B=@n29.RGB, Alpha=@n78)
  n82 = SetMaterialAttributes!(
   MaterialAttributes=@n2.MaterialAttributes
   BaseColor=LinearInterpolate(A=LinearInterpolate(A=@n79, B=@n80, Alpha=@n50), B=Multiply(A=@n33, B=@n81), Alpha=@n33)
   Roughness=StaticSwitchParameter(True=@n1, False=LinearInterpolate(A=@n73, B=@n74, Alpha=@n75)) {group=NeonSign, parameterName=Use other Function as Base, sortPriority=-1}
   Normal=StaticSwitchParameter(True=@n1, False=StaticSwitchParameter(True=@n2.Normal, False=@n28) {group=NeonSign, parameterName=Use No Normal, sortPriority=4}) {defaultValue=true, group=NeonSign, parameterName=Use other Function as Base, sortPriority=-1}
   Specular=StaticSwitchParameter(True=@n1, False=LinearInterpolate(A=@n70, B=@n71, Alpha=@n72)) {defaultValue=true, group=NeonSign, parameterName=Use other Function as Base, sortPriority=-1}
   Metallic=StaticSwitchParameter(True=@n1, False=ScalarParameter() {group=NeonSign, parameterName=NS_Metallic, sortPriority=7}) {defaultValue=true, group=NeonSign, parameterName=Use other Function as Base, sortPriority=-1}
   EmissiveColor=LinearInterpolate(
    A=Add(A=@n55, B=@n61)
    B=StaticSwitchParameter(True=@n66, False=@n64) {defaultValue=true, group=NeonSign, parameterName=Use Texture as Reactive Multiplier, sortPriority=12}
    Alpha=LinearInterpolate(A=@n68, B=@n69, Alpha=@n69)
   )
  ) {attributeSetTypes [
    #1
    #2
    #3
    #4
    #5
    #6
   ]}
  n83 = FunctionInput() {bUsePreviewValueAsDefault=true, inputName="Customization Mask (_CM B Top left)", inputType=FunctionInput_Scalar, sortPriority=24}
  n84 = FunctionInput() {bUsePreviewValueAsDefault=true, inputName="Customization Mask 2 (_CM B Top Right)", inputType=FunctionInput_Scalar, sortPriority=25}
  n85 = Add(A=@n84, B=@n83) {desc="Base + Secondary = exposed primary"}
  n86 = OneMinus(None=@n85) {desc=Primary}
  n87 = LinearInterpolate(
   A=LinearInterpolate(A=@n83, B=@n84, Alpha=ScalarParameter() {group=NeonSign, parameterName=MASK_OnlySecondary, sortPriority=-99})
   B=@n86
   Alpha=ScalarParameter() {group=NeonSign, parameterName=MASK_OnlyPrimary, sortPriority=-98}
  )
  n88 = ScalarParameter() {group=NeonSign, parameterName=MASK_BaseAndSecondary, sortPriority=-97}
  n89 = LinearInterpolate(
   A=LinearInterpolate(
    A=LinearInterpolate(A=@n87, B=@n85, Alpha=@n88)
    B=Add(A=@n86, B=@n83)
    Alpha=ScalarParameter() {group=NeonSign, parameterName=MASK_BaseAndPrimary, sortPriority=-96}
   )
   B=Add(A=@n86, B=@n84)
   Alpha=ScalarParameter() {group=NeonSign, parameterName=MASK_SecondaryAndPrimary, sortPriority=-95}
  )
  n90 = ScalarParameter() {group=NeonSign, parameterName=MASK_CoverEntireWeapon, sortPriority=-100}
  n91 = StaticSwitchParameter(True=BlendMaterialAttributes(A=@n2.MaterialAttributes, B=@n82, Alpha=LinearInterpolate(A=@n89, B=@n90, Alpha=@n90)), False=@n1) {
   desc=TopSide Projection Base - Mask Driven with separate channels driving a Color - Channels glow offset timings - Reactive -Brighten all at once on Fire
   group=Customization_Wrap
   parameterName=Wrap_NeonSign
  }
  FunctionOutput(None=@n91)
 ]}
