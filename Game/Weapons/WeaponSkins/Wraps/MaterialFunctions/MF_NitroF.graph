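# Material function for the animated, reactive NitroFlow wrap, blending base and accent materials by part masks behind the Wrap_NitroFlow static switch.
# Pre-skinned bounds and normals drive projected UVs, with scale, rotation and offset selected by item-state inputs for weapons, vehicles, characters and other items.
# Panning, warped, mirrored noise forms curve-atlas-colored spots, layered with an icon, normals and Fresnel thin film.
# A Blueprint timestamp parameter drives timed brightness fades.
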MaterialFunction /Game/Weapons/WeaponSkins/Wraps/MaterialFunctions/MF_NitroF
graph {nodes [
  n1 = Add(A=NamedRerouteUsage(), B=ComponentMask(None=NamedRerouteUsage()) {r=true})
  n2 = ComponentMask(None=NamedRerouteUsage()) {a=true}
  n3 = Saturate(None=LinearInterpolate(B=Saturate(None=@n1), Alpha=@n2) {constA=1})
  n4 = LinearInterpolate(A=NamedRerouteUsage(), Alpha=@n2) {constA=1}
  n5 = NamedRerouteUsage()
  n6 = StaticBool() {value=true}
  n7 = MaterialFunctionCall!(
   Normal_Map=NamedRerouteUsage()
   Offset Rt
   Pivot
   Caculated_Bounds=NamedRerouteUsage()
   Pre_Skinned_Normals=NamedRerouteUsage()
   is_State=NamedRerouteUsage()
   UV_Rotation=NamedRerouteUsage()
   Uvs_Offset=ComponentMask(None=@n5) {b=true, g=true, r=true}
   Uvs_TileRate=ComponentMask(None=@n5) {a=true}
   Uvs_Offset_Pivot
   For_Safe_Zones=@n6
   Rotation_Comp
   Clamp Y=@n6
   Clamp X=@n6
   Clamp Padding_Comp
   No_Rotation
  ) {materialFunction=&/MF_Projection_Side_Normal_V2}
  n8 = NamedRerouteDeclaration(None=@n7.UVs) {name=Uvs_1, nodeColor=<rgba 0 1 0.3672 1>}
  n9 = NamedRerouteUsage()
  n10 = MaterialFunctionCall(Float4=Multiply(A=Multiply(A=@n3, B=@n4), B=TextureSample(UVs=@n8, Tex=@n9) {samplerType=SAMPLERTYPE_Normal}.RGBA)) {materialFunction=&BreakOutFloat4Components}
  Add(
   A=@n10.A
   B=Add(A=NamedRerouteUsage(), B=ScalarParameter() {defaultValue=-0.5, group=Wrap_NitroFlow_Icon, parameterName=NiFlow_Icon_Noise_Amount})
  )
  n11 = NamedRerouteUsage()
  ComponentMask(None=@n11) {a=true}
  n12 = TextureSample() {samplerType=SAMPLERTYPE_Masks, texture=/Game/Weapons/FORT_Rifles/Textures/T_Scar_CM}
  n13 = FunctionInput(Preview=@n12.R) {bUsePreviewValueAsDefault=true, inputName=Mask Primary, inputType=FunctionInput_Scalar, sortPriority=24, previewValue=<1 0 0 1>}
  n14 = FunctionInput(Preview=@n12.B) {bUsePreviewValueAsDefault=true, inputName=Mask Secondary, inputType=FunctionInput_Scalar, sortPriority=24, previewValue=<1 0 0 1>}
  n15 = BlendMaterialAttributes(A=NamedRerouteUsage(), B=NamedRerouteUsage(), Alpha=Saturate(None=Add(A=@n13, B=@n14))) {vertexAttributeBlendType=UseA}
  n16 = NamedRerouteUsage()
  n17 = NamedRerouteDeclaration(None=@n14) {name=Msk_Accent, nodeColor=<rgba 0 1 0.7656 1>}
  FunctionOutput(None=StaticSwitchParameter(
    True=BlendMaterialAttributes(A=@n15, B=@n16, Alpha=@n17) {vertexAttributeBlendType=UseA}
    False=FunctionInput() {inputName=Material, inputType=FunctionInput_MaterialAttributes}
   ) {group=Customization_Wrap, parameterName=Wrap_NitroFlow})
  MaterialFunctionCall(Float2=NamedRerouteUsage()) {materialFunction=&BreakOutFloat2Components}
  MaterialFunctionCall(Float2=NamedRerouteUsage()) {materialFunction=&BreakOutFloat2Components}
  MaterialFunctionCall(Float2=NamedRerouteUsage()) {materialFunction=&BreakOutFloat2Components}
  MaterialFunctionCall(Float2=NamedRerouteUsage()) {materialFunction=&BreakOutFloat2Components}
  Multiply(A=MaterialFunctionCall(Value_Morning=Constant() {r=0.5}, Value_Day=Constant() {r=1}, Value_Evening=Constant() {r=0.5}, Value_Night=Constant() {r=0.25}) {materialFunction=&Scalar_Time_of_Day}.Result)
  n18 = VectorParameter() {group=Wrap_NitroFlow_Back, parameterName=NiFlow_Spots_Clamp_Noise, defaultValue=<rgba 0.4 0 0 1>}
  n19 = NamedRerouteUsage()
  n20 = NamedRerouteUsage()
  n21 = ComponentMask(None=Multiply(A=Time(), B=VectorParameter() {group=Wrap_NitroFlow_Back, parameterName=NiFlow_Spots_Speed_A, defaultValue=<rgba -0.0063 0.005 0 1>}.RGB)) {g=true, r=true}
  n22 = Subtract(A=NamedRerouteUsage(), B=Constant2Vector() {g=0.5})
  n23 = ComponentMask(None=@n22) {g=true}
  n24 = Multiply(A=NamedRerouteUsage(), B=AppendVector(A=ComponentMask(None=@n22) {r=true}, B=Abs(None=@n23)))
  n25 = NamedRerouteUsage()
  n26 = Multiply(A=@n24, B=ScalarParameter() {defaultValue=1, group=Wrap_NitroFlow_Back, parameterName=NiFlow_Spots_TileRate_B})
  n27 = ComponentMask(None=Multiply(A=Time(), B=VectorParameter() {group=Wrap_NitroFlow_Back, parameterName=NiFlow_Spots_Speed_B, defaultValue=<rgba 0.0073 0.0053 0 1>}.RGB)) {g=true, r=true}
  n28 = Add(
   A=TextureSample(UVs=Add(A=@n24, B=@n21), Tex=@n25) {samplerType=SAMPLERTYPE_LinearColor}.R
   B=TextureSample(UVs=Add(A=@n26, B=@n27), Tex=@n25) {samplerType=SAMPLERTYPE_LinearColor}.B
  )
  n29 = Multiply(A=@n24, B=ScalarParameter() {defaultValue=1, group=Wrap_NitroFlow_Back, parameterName=NiFlow_Spots_Warp_TileRate})
  n30 = ComponentMask(None=Multiply(A=Time(), B=VectorParameter() {group=Wrap_NitroFlow_Back, parameterName=NiFlow_Spots_Warp_Speed, defaultValue=<rgba -0.01 -0.02 0 1>}.RGB)) {g=true, r=true}
  n31 = Add(A=TextureSample(UVs=Add(A=@n29, B=@n30), Tex=NamedRerouteUsage()) {samplerType=SAMPLERTYPE_LinearColor, texture=/Game/Weapons/WeaponSkins/Wraps/Textures/NitroFlow/T_Wrap_NitroFlow_Msk}.A) {constB=0}
  n32 = ScalarParameter() {defaultValue=0.3, group=Wrap_NitroFlow_Back, parameterName=NiFlow_Spots_Warp_Amount}
  n33 = Multiply(A=@n24, B=ScalarParameter() {defaultValue=0.9, group=Wrap_NitroFlow_Back, parameterName=NiFlow_Spots_TileRate_C}) {constB=0.9}
  n34 = ComponentMask(None=Multiply(A=Time(), B=VectorParameter() {group=Wrap_NitroFlow_Back, parameterName=NiFlow_Spots_Speed_C, defaultValue=<rgba -0.005 -0.0095 0 1>}.RGB)) {g=true, r=true}
  n35 = TextureSample(UVs=Add(A=Multiply(A=@n31, B=@n32) {constB=0.3}, B=Add(A=@n33, B=@n34)), Tex=@n25) {samplerType=SAMPLERTYPE_LinearColor}
  n36 = Multiply(
   A=LinearInterpolate(
    A=NamedRerouteUsage()
    Alpha=MaterialFunctionCall(Float4=@n19) {materialFunction=&BreakOutFloat4Components}.A
   ) {constA=1}
   B=Multiply(A=OneMinus(None=@n20), B=Add(A=@n28, B=@n35.R))
  ) {constA=1}
  n37 = SmoothStep(Min=@n18.R, Max=@n18.R, Value=@n36) {constMax=0.4, constMin=0.3}
  Multiply(A=@n37)
  n38 = Add(A=NamedRerouteUsage(), B=ScalarParameter() {defaultValue=0.4, group=Wrap_NitroFlow_Back, parameterName=NiFlow_Back_Noise_Amount})
  n39 = ComponentMask(None=NamedRerouteUsage()) {r=true}
  n40 = OneMinus(None=@n39)
  n41 = Saturate(None=Divide(A=Add(A=@n40) {constB=-0.25}) {constB=0.25})
  n42 = Multiply(A=LinearInterpolate(B=SmoothStep(Value=@n38), Alpha=Power(Base=@n41)) {constA=1, constB=0}, B=@n39)
  n43 = CurveAtlasRowParameter(CurveTime=@n42) {
   atlas=/Game/Weapons/WeaponSkins/Wraps/Textures/CruveAtlas_Wraps_01
   curve=/Game/Weapons/WeaponSkins/Wraps/Textures/0_ColorCurves/CC_NiFlow_A
   defaultValue=120
   group=Wrap_NitroFlow_Back
   parameterName=NiFlow_Back_CColors
  }
  n44 = MaterialFunctionCall!(
   Index=ScalarParameter() {defaultValue=120, group=Wrap_NitroFlow_Back, parameterName="NiFlow_Back_CColors  Index"}
   CCinput=@n43
   CCinput_Alpha
   UV=@n42
  ) {materialFunction=&/MF_Global_CC_Tex}
  Multiply(A=@n44.3V)
  n45 = Subtract(A=Time(), B=ScalarParameter() {defaultValue=-9999999, group=Wrap_Reactive, parameterName=TimeFrom-BP----------------, sortPriority=600})
  n46 = Divide(A=@n45, B=ScalarParameter() {defaultValue=0.25, group=Wrap_NitroFlow_Reactive, parameterName=NiFlow_Brightness_Duration, sortPriority=90})
  n47 = LinearInterpolate(
   A=ScalarParameter() {defaultValue=2, group=Wrap_NitroFlow_Reactive, parameterName=NiFlow_Brightness_Start, sortPriority=20}
   B=ScalarParameter() {group=Wrap_NitroFlow_Reactive, parameterName=NiFlow_Brightness_End, sortPriority=20}
   Alpha=Power(
    Base=Saturate(None=@n46)
    Exp=ScalarParameter() {defaultValue=0.25, group=Wrap_NitroFlow_Reactive, parameterName=NiFlow_Brightness_Falloff, sortPriority=90}
   )
  )
  n48 = Multiply(A=Time(), B=ScalarParameter() {defaultValue=0.5, group=Wrap_NitroFlow_Back, parameterName=NiFlow_Back_Reactive_Noise_Speed})
  n49 = Multiply(
   A=SphereMask(A=NamedRerouteUsage(), B=Constant() {r=0.5}, Radius=Constant() {r=1}, Hardness=Constant())
   B=ScalarParameter() {defaultValue=20, group=Wrap_NitroFlow_Back, parameterName=NiFlow_Back_Reactive_Noise_TileRate}
  ) {constB=20}
  n50 = Sine(None=Frac(None=Add(A=@n48, B=@n49)))
  n51 = SmoothStep(Value=@n50) {constMin=-1}
  n52 = Add(A=Multiply(A=Multiply(A=@n47, B=@n51)) {constB=2})
  Multiply(A=MaterialFunctionCall!(X range Y range, Gradient Height, ShowGrid?, Thickness, Color, "f(x)"=@n52) {materialFunction=/Engine/Functions/Engine_MaterialFunctions01/Debug/PlotFunctionOnGraph}."f(x) = y")
  Multiply(A=@n50)
  n53 = VectorParameter() {group=Wrap_NitroFlow_Accent, parameterName=NiFlow_Accent_BaseColor, defaultValue=<rgba 0.0076 0.0094 0.015 1>}
  n54 = SetMaterialAttributes!(
   MaterialAttributes=NamedRerouteUsage()
   BaseColor=@n53.RGB
   Roughness=ScalarParameter() {defaultValue=0.4, group=Wrap_NitroFlow_Accent, parameterName=NiFlow_Accent_Roughness}
   Specular=ScalarParameter() {defaultValue=0.25, group=Wrap_NitroFlow_Accent, parameterName=NiFlow_Accent_Spec}
   Metallic=ScalarParameter() {group=Wrap_NitroFlow_Accent, parameterName=NiFlow_Accent_Metalic}
   Normal
   EmissiveColor
  ) {attributeSetTypes [
    #1
    #2
    #3
    #4
    #5
    #6
   ]}
  NamedRerouteDeclaration(None=@n54) {name=Mat_Accent, nodeColor=<rgba 0 0.3906 1 1>}
  n55 = FunctionInput(Preview=@n12.G) {bUsePreviewValueAsDefault=true, inputName=Mask Crevise, inputType=FunctionInput_Scalar, sortPriority=24, previewValue=<1 0 0 1>}
  NamedRerouteDeclaration(None=@n55) {name=Msk_Curvature, nodeColor=<rgba 0 1 0.7422 1>}
  n56 = TextureSample() {samplerType=SAMPLERTYPE_Normal, texture=/Game/Weapons/FORT_Rifles/T_SCAR_N}
  n57 = Constant3Vector() {constant=<rgba 1 0 0 0>}
  n58 = GetMaterialAttributes(None=FunctionInput(Preview=SetMaterialAttributes!(MaterialAttributes, Normal=@n56.RGB, BaseColor=@n57) {attributeSetTypes [#5, #1]}) {inputName=Material_Root, inputType=FunctionInput_MaterialAttributes}) {attributeGetTypes [
    #1
    #2
    #5
    #4
   ]}
  NamedRerouteDeclaration(None=@n58.MaterialAttributes) {name=Root_Material, nodeColor=<rgba 0.7812 1 0 1>}
  n59 = Multiply(A=NamedRerouteUsage(), B=ScalarParameter() {defaultValue=1, group=Wrap_NitroFlow_Back, parameterName=NiFlow_Back_Noise_TileRate})
  n60 = ComponentMask(None=Multiply(A=Time(), B=VectorParameter() {group=Wrap_NitroFlow_Back, parameterName=NiFlow_Back_Noise_Speed, defaultValue=<rgba -0.01 -0.02 0 1>}.RGB)) {g=true, r=true}
  NamedRerouteDeclaration(None=TextureSample(UVs=Add(A=@n59, B=@n60), Tex=NamedRerouteUsage()) {samplerType=SAMPLERTYPE_LinearColor, texture=&T_Wrap_NitroFlow_Msk}.A) {name=Msk_Noise, nodeColor=<rgba 1 0.1641 0 1>}
  NamedRerouteDeclaration(None=TextureObjectParameter() {group=Wrap_NitroFlow, parameterName=NiFlow__Tex_Msk2, texture=&T_BlacksRGB}) {name=Tex_Msk2, nodeColor=<rgba 0 1 0.3672 1>}
  NamedRerouteDeclaration(None=@n24) {name=UV_Mirror, nodeColor=<rgba 0 0.25 1 1>}
  n61 = NamedRerouteUsage()
  n62 = Multiply(
   A=NamedRerouteUsage()
   B=Multiply(
    A=ScalarParameter() {defaultValue=5, group=Wrap_NitroFlow_Icon, parameterName=NiFlow_Icon_Noise_TileRate}
    B=ComponentMask(None=@n61) {a=true}
   )
  )
  n63 = ComponentMask(None=Multiply(A=Time(), B=VectorParameter() {group=Wrap_NitroFlow_Icon, parameterName=NiFlow_Icon_Noise_Speed, defaultValue=<rgba 0.023 -0.043 0 1>}.RGB)) {g=true, r=true}
  NamedRerouteDeclaration(None=TextureSample(UVs=Add(A=@n62, B=@n63), Tex=NamedRerouteUsage()) {samplerType=SAMPLERTYPE_LinearColor, texture=&T_Wrap_NitroFlow_Msk}.A) {name=Msk_Noise_B, nodeColor=<rgba 1 0.1641 0 1>}
  NamedRerouteDeclaration(None=@n47) {name=React_Back, nodeColor=<rgba 1 0.2109 0 1>}
  n64 = ScalarParameter() {defaultValue=1.5, group=Wrap_NitroFlow_Uvs, parameterName=NiFlow_Spot_Scale_Vehicle, sortPriority=150}
  n65 = ScalarParameter() {defaultValue=1.5, group=Wrap_NitroFlow_Uvs, parameterName=NiFlow_Spot_Scale_Vehicle_Side, sortPriority=150}
  n66 = MaterialFunctionCall(Float4=NamedRerouteUsage()) {materialFunction=&BreakOutFloat4Components}
  n67 = LinearInterpolate(
   A=LinearInterpolate(
    A=ScalarParameter() {defaultValue=1.5, group=Wrap_NitroFlow_Uvs, parameterName=NiFlow_Spot_Scale_Weapon, sortPriority=150}
    B=LinearInterpolate(A=@n64, B=@n65, Alpha=@n66.G)
    Alpha=@n66.R
   )
   B=LinearInterpolate(
    A=ScalarParameter() {defaultValue=1.5, group=Wrap_NitroFlow_Uvs, parameterName=NiFlow_Spot_Scale_Mech, sortPriority=150}
    B=ScalarParameter() {defaultValue=1.5, group=Wrap_NitroFlow_Uvs, parameterName=NiFlow_Spot_Scale_Character, sortPriority=150}
    Alpha=@n66.A
   )
   Alpha=@n66.B
  )
  n68 = ScalarParameter() {defaultValue=1.5, group=Wrap_NitroFlow_Uvs, parameterName=NiFlow_Spot_Scale_Backpack, sortPriority=150}
  n69 = MaterialFunctionCall(Float2=NamedRerouteUsage()) {materialFunction=&BreakOutFloat2Components}
  NamedRerouteDeclaration(None=LinearInterpolate(
    A=LinearInterpolate(A=@n67, B=@n68, Alpha=@n69.R)
    B=ScalarParameter() {defaultValue=1.5, group=Wrap_NitroFlow_Uvs, parameterName=NiFlow_Spot_Scale_Pickaxe, sortPriority=150}
    Alpha=@n69.G
   )) {name=UVs_Scale_Spots, nodeColor=<rgba 0 0.7188 1 1>}
  n70 = VectorParameter() {group=Wrap_NitroFlow_Back, parameterName=NiFlow_Spots_Color_Bleed, defaultValue=<rgba 0.0062 0.008 0.0521 0.5>}
  n71 = LinearInterpolate(A=@n70.RGB, B=@n44.3V, Alpha=LinearInterpolate(B=@n37, Alpha=@n70.A) {constA=1})
  n72 = ScalarParameter() {defaultValue=0.1, group=Wrap_NitroFlow_Back, parameterName=NiFlow_Spots_Darkness}
  n73 = VectorParameter() {group=Wrap_NitroFlow_Back, parameterName=NiFlow_Spots_Clamp_Blacks, defaultValue=<rgba 0.4 0.3 0 1>}
  n74 = SmoothStep(Min=@n73.R, Max=@n73.R, Value=@n36) {constMax=0.4, constMin=0.3}
  NamedRerouteDeclaration(None=Multiply(A=Multiply(A=@n52, B=@n71) {constA=1}, B=LinearInterpolate(A=@n72, Alpha=@n74) {constA=0.5})) {name=Spots, nodeColor=<rgba 0.1719 1 0 1>}
  n75 = Multiply(A=NamedRerouteUsage(), B=NamedRerouteUsage())
  n76 = Add(A=NamedRerouteUsage()) {constB=-0.5}
  n77 = LinearInterpolate(
   A=@n75
   B=Multiply(A=@n75, B=ScalarParameter() {defaultValue=3, group=Wrap_NitroFlow, parameterName=NiFlow_Edge_Brightness})
   Alpha=Clamp(None=@n76) {minDefault=-0.15}
  )
  n78 = LinearInterpolate(
   A=ScalarParameter() {defaultValue=0.05, group=Wrap_NitroFlow_Spots, parameterName=NiFlow_Spots_Mult_Max}
   B=ScalarParameter() {defaultValue=0.25, group=Wrap_NitroFlow_Spots, parameterName=NiFlow_Spots_Mult_Min}
   Alpha=NamedRerouteUsage()
  )
  n79 = ScalarParameter() {defaultValue=0.5, group=Wrap_NitroFlow_ThinFlim, parameterName=NiFlow_ThinFilm_Warp}
  n80 = Frac(None=Multiply(
    A=ScalarParameter() {defaultValue=1, group=Wrap_NitroFlow_ThinFlim, parameterName=NiFlow_ThinFilm_Scale}
    B=Fresnel(ExponentIn=@n79) {baseReflectFraction=0}
   ))
  n81 = CurveAtlasRowParameter(CurveTime=@n80) {
   atlas=&CruveAtlas_Wraps_01
   curve=/Game/Weapons/WeaponSkins/Wraps/Textures/0_ColorCurves/CC_NiFlow_C
   defaultValue=128
   group=Wrap_NitroFlow_ThinFlim
   parameterName=NiFlow_ThinFlim_CColors
  }
  n82 = Multiply(
   A=MaterialFunctionCall(
    Index=ScalarParameter() {defaultValue=128, group=Wrap_NitroFlow_ThinFlim, parameterName="NiFlow_ThinFlim_CColors  Index"}
    CCinput=@n81
    UV=@n80
   ) {materialFunction=&MF_Global_CC_Tex}.3V
   B=ScalarParameter() {defaultValue=0.5, group=Wrap_NitroFlow_ThinFlim, parameterName=NiFlow_ThinFilm_Intensity}
  )
  n83 = ScalarParameter() {defaultValue=1, group=Wrap_NitroFlow_ThinFlim, parameterName=NiFlow_FilmSpot_Msk_Brightness}
  n84 = Add(A=NamedRerouteUsage(), B=ScalarParameter() {defaultValue=0.1, group=Wrap_NitroFlow_ThinFlim, parameterName=NiFlow_FilmSpot_Big_Msk})
  n85 = OneMinus(None=NamedRerouteUsage())
  n86 = LinearInterpolate(
   A=ScalarParameter() {defaultValue=1, group=Wrap_NitroFlow_ThinFlim, parameterName=NiFlow_FilmSpots_Mult_Max}
   B=ScalarParameter() {defaultValue=1, group=Wrap_NitroFlow_ThinFlim, parameterName=NiFlow_FilmSpots_Mult_Min}
   Alpha=NamedRerouteUsage()
  )
  n87 = Multiply(A=Saturate(None=Multiply(A=@n83, B=@n84)), B=Saturate(None=Add(A=@n85, B=@n86)))
  n88 = Add(
   A=LinearInterpolate(A=@n77, B=Multiply(A=@n78, B=@n77), Alpha=NamedRerouteUsage())
   B=Multiply(A=NamedRerouteUsage(), B=Multiply(A=@n82, B=@n87))
  )
  n89 = NamedRerouteUsage()
  n90 = NamedRerouteUsage()
  n91 = NamedRerouteUsage()
  n92 = NamedRerouteUsage()
  n93 = Multiply(A=@n88, B=ScalarParameter() {defaultValue=0.25, group=Wrap_NitroFlow, parameterName=NiFlow_BaseColor_Emis_Amount}) {constB=0.5}
  n94 = NamedRerouteUsage()
  n95 = ScalarParameter() {defaultValue=0.3, group=Wrap_NitroFlow, parameterName=NiFlow_Roughness_Icon}
  n96 = ScalarParameter() {defaultValue=0.25, group=Wrap_NitroFlow, parameterName=NiFlow_Roughness}
  n97 = NamedRerouteUsage()
  n98 = SetMaterialAttributes!(
   MaterialAttributes=NamedRerouteUsage()
   BaseColor=Multiply(A=ScalarParameter() {defaultValue=0.5, group=Wrap_NitroFlow, parameterName=NiFlow_BaseColor_Amount}, B=@n88)
   Roughness=Add(A=LinearInterpolate(A=@n95, B=@n96, Alpha=@n97), B=Saturate(None=@n76))
   Specular=ScalarParameter() {defaultValue=0.25, group=Wrap_NitroFlow, parameterName=NiFlow_Spec}
   Metallic=ScalarParameter() {group=Wrap_NitroFlow, parameterName=NiFlow_Metalic}
   Normal=Add(A=Add(A=@n89, B=@n90), B=Multiply(A=@n91, B=@n92))
   EmissiveColor=Clamp(None=Add(A=@n93, B=@n94)) {maxDefault=20}
  ) {attributeSetTypes [
    #1
    #2
    #3
    #4
    #5
    #6
   ]}
  NamedRerouteDeclaration(None=@n98) {name=Mat_Base, nodeColor=<rgba 0 0.3906 1 1>}
  n99 = ScalarParameter() {group=Wrap_NitroFlow_Uvs, parameterName="NiFlow_Spot_Rotation_Weapon\\Pickaxe", sortPriority=150}
  n100 = MaterialFunctionCall(Float4=NamedRerouteUsage()) {materialFunction=&BreakOutFloat4Components}
  n101 = LinearInterpolate(
   A=ScalarParameter() {group=Wrap_NitroFlow_Uvs, parameterName=NiFlow_Spot_Rotation_Vehicle, sortPriority=150}
   B=ScalarParameter() {group=Wrap_NitroFlow_Uvs, parameterName=NiFlow_Spot_Rotation_Vehicle_Side, sortPriority=150}
   Alpha=@n100.G
  )
  NamedRerouteDeclaration(None=LinearInterpolate(
    A=LinearInterpolate(A=@n99, B=@n101, Alpha=@n100.R)
    B=ScalarParameter() {group=Wrap_NitroFlow_Uvs, parameterName="NiFlow_Spot_Rotation_Mech\\Character\\Back", sortPriority=150}
    Alpha=@n100.B
   )) {name=UVs_Rotation_Spots, nodeColor=<rgba 0 0.7188 1 1>}
  n102 = VectorParameter() {group=Wrap_NitroFlow_Uvs, parameterName=NiFlow_Uv_Offset, sortPriority=0, defaultValue=<rgba 0 0 0 1>}
  NamedRerouteDeclaration(None=MaterialFunctionCall(
    Caculated_Bounds=NamedRerouteUsage()
    Pre_Skinned_Normals=NamedRerouteUsage()
    is_State=NamedRerouteUsage()
    UV_Rotation=NamedRerouteUsage()
    Uvs_TileRate=ScalarParameter() {defaultValue=1, group=Wrap_NitroFlow_Uvs, parameterName=NiFlow_Uv_Scale, sortPriority=0}
    Uvs_Offset_Pivot=@n102.RGB
    No_Rotation=StaticBoolParameter() {defaultValue=true, group=Wrap_NitroFlow_Uvs, parameterName=NiFlow_Spots_No_Rotation}
   ) {materialFunction=&MF_Projection_Side_Normal_V2}.UVs) {name=Uvs, nodeColor=<rgba 0 1 0.3672 1>}
  n103 = StaticBool() {value=true}
  NamedRerouteDeclaration(None=MaterialFunctionCall(
    Normal_Map=NamedRerouteUsage()
    Caculated_Bounds=NamedRerouteUsage()
    Pre_Skinned_Normals=NamedRerouteUsage()
    is_State=NamedRerouteUsage()
    UV_Rotation=NamedRerouteUsage()
    Uvs_Offset=ComponentMask(None=@n11) {b=true, g=true, r=true}
    Uvs_TileRate=Constant() {r=3}
    For_Safe_Zones=@n103
    Clamp Y=@n103
    Clamp X=@n103
   ) {materialFunction=&MF_Projection_Side_Normal_V2}.UVs) {name=Uvs_2, nodeColor=<rgba 0 1 0.3672 1>}
  n104 = VectorParameter() {group=Wrap_NitroFlow_Uvs, parameterName=NiFlow_Offset_Scale_Weapon, defaultValue=<rgba -0.31 0 0 0.55>}
  n105 = VectorParameter() {group=Wrap_NitroFlow_Uvs, parameterName=NiFlow_Offset_Scale_Vehicle, defaultValue=<rgba 0 0 0 1>}
  n106 = AppendVector(A=@n105.RGB, B=@n105.RGB)
  n107 = VectorParameter() {group=Wrap_NitroFlow_Uvs, parameterName=NiFlow_Offset_Scale_VehicleSide, defaultValue=<rgba 0 0 0 1>}
  n108 = AppendVector(A=@n107.RGB, B=@n107.RGB)
  n109 = MaterialFunctionCall(Float4=NamedRerouteUsage()) {materialFunction=&BreakOutFloat4Components}
  n110 = VectorParameter() {group=Wrap_NitroFlow_Uvs, parameterName=NiFlow_Offset_Scale_Mech, defaultValue=<rgba 0 0 0 1>}
  n111 = VectorParameter() {group=Wrap_NitroFlow_Uvs, parameterName=NiFlow_Offset_Scale_Character, defaultValue=<rgba 0 0 0 1>}
  n112 = LinearInterpolate(
   A=LinearInterpolate(A=AppendVector(A=@n104.RGB, B=@n104.RGB), B=LinearInterpolate(A=@n106, B=@n108, Alpha=@n109.G), Alpha=@n109.R)
   B=LinearInterpolate(A=AppendVector(A=@n110.RGB, B=@n110.RGB), B=AppendVector(A=@n111.RGB, B=@n111.RGB), Alpha=@n109.A)
   Alpha=@n109.B
  )
  n113 = VectorParameter() {group=Wrap_NitroFlow_Uvs, parameterName=NiFlow_Offset_Scale_Backpack, defaultValue=<rgba 0 0 0 1>}
  n114 = AppendVector(A=@n113.RGB, B=@n113.RGB)
  n115 = MaterialFunctionCall(Float2=NamedRerouteUsage()) {materialFunction=&BreakOutFloat2Components}
  n116 = VectorParameter() {group=Wrap_NitroFlow_Uvs, parameterName=NiFlow_Offset_Scale_Pickaxe, defaultValue=<rgba 0 0 0 1>}
  NamedRerouteDeclaration(None=LinearInterpolate(A=LinearInterpolate(A=@n112, B=@n114, Alpha=@n115.R), B=AppendVector(A=@n116.RGB, B=@n116.RGB), Alpha=@n115.G)) {name=UVs_Offset_Scale, nodeColor=<rgba 0 0.625 1 1>}
  n117 = MaterialFunctionCall(Color=Abs(None=NamedRerouteUsage())) {materialFunction=&SplitComponents}
  n118 = MaterialFunctionCall(Float4=NamedRerouteUsage()) {materialFunction=&BreakOutFloat4Components}
  n119 = LinearInterpolate(A=@n117.R, B=LinearInterpolate(A=@n117.B, B=@n117.B, Alpha=@n118.G), Alpha=@n118.R)
  n120 = ScalarParameter() {defaultValue=1, group=Wrap_NitroFlow, parameterName=NiFlow_TriBlend_Contrast}
  NamedRerouteDeclaration(None=MaterialFunctionCall(In=LinearInterpolate(A=@n119, B=@n117.G, Alpha=@n118.B), Contrast=@n120) {materialFunction=&CheapContrast}.Result) {name=Tri_Mask, nodeColor=<rgba 1 0.4688 0 1>}
  n121 = ScalarParameter() {group=Wrap_NitroFlow_Uvs, parameterName="NiFlow_Rotation_Weapon\\Pickaxe", sortPriority=150}
  n122 = MaterialFunctionCall(Float4=NamedRerouteUsage()) {materialFunction=&BreakOutFloat4Components}
  n123 = LinearInterpolate(
   A=ScalarParameter() {group=Wrap_NitroFlow_Uvs, parameterName=NiFlow_Rotation_Vehicle, sortPriority=150}
   B=ScalarParameter() {group=Wrap_NitroFlow_Uvs, parameterName=NiFlow_Rotation_Vehicle_Side, sortPriority=150}
   Alpha=@n122.G
  )
  NamedRerouteDeclaration(None=LinearInterpolate(
    A=LinearInterpolate(A=@n121, B=@n123, Alpha=@n122.R)
    B=ScalarParameter() {group=Wrap_NitroFlow_Uvs, parameterName="NiFlow_Rotation_Mech\\Character\\Back", sortPriority=150}
    Alpha=@n122.B
   )) {name=UVs_Rotation, nodeColor=<rgba 0 0.7188 1 1>}
  n124 = TextureObjectParameter() {
   group=Wrap_NitroFlow
   parameterName=NiFlow__Tex_N
   samplerType=SAMPLERTYPE_Normal
   texture=&FlatNormalBlue
  }
  NamedRerouteDeclaration(None=@n124) {name=Tex_N, nodeColor=<rgba 0 1 0.3672 1>}
  NamedRerouteDeclaration(None=Multiply(A=@n4, B=@n7.Raw_Normal)) {name=Icon_Normal, nodeColor=<rgba 1 0 0.9844 1>}
  NamedRerouteDeclaration(None=@n10.R) {name=Icon_Msk, nodeColor=<rgba 0.2656 1 0 1>}
  n125 = Multiply(A=@n10.G, B=@n10.G)
  n126 = CurveAtlasRowParameter(CurveTime=@n10.A) {
   atlas=&CruveAtlas_Wraps_01
   curve=/Game/Weapons/WeaponSkins/Wraps/Textures/0_ColorCurves/CC_NiFlow_B
   defaultValue=121
   group=Wrap_NitroFlow_Icon
   parameterName=NiFlow_Icon_CColors
  }
  n127 = MaterialFunctionCall(Index=ScalarParameter() {defaultValue=121, group=Wrap_NitroFlow_Icon, parameterName="NiFlow_Icon_CColors  Index"}, CCinput=@n126, UV=@n10.A) {materialFunction=&MF_Global_CC_Tex}
  n128 = ScalarParameter() {defaultValue=2, group=Wrap_NitroFlow_Reactive, parameterName=NiFlow_Brightness_Icon_Start, sortPriority=10}
  n129 = ScalarParameter() {defaultValue=0.5, group=Wrap_NitroFlow_Reactive, parameterName=NiFlow_Brightness_Icon_End, sortPriority=10}
  n130 = Saturate(None=Divide(
    A=@n45
    B=ScalarParameter() {defaultValue=0.25, group=Wrap_NitroFlow_Reactive, parameterName=NiFlow_Brightness_Icon_Duration, sortPriority=90}
   ))
  NamedRerouteDeclaration(None=Multiply(A=Multiply(A=@n125, B=@n127.3V), B=LinearInterpolate(A=@n128, B=@n129, Alpha=@n130))) {name=Icon_Msk_B, nodeColor=<rgba 0.2656 1 0 1>}
  NamedRerouteDeclaration(None=OneMinus(None=@n125)) {name=Icon_Msk_C, nodeColor=<rgba 0.2656 1 0 1>}
  n131 = TextureObjectParameter() {
   group=Wrap_NitroFlow
   parameterName=NiFlow__Tex_N2
   samplerType=SAMPLERTYPE_Normal
   texture=&FlatNormalBlue
  }
  NamedRerouteDeclaration(None=@n131) {name=Tex_N_1, nodeColor=<rgba 0 1 0.3672 1>}
  NamedRerouteDeclaration(None=TextureObjectParameter() {group=Wrap_NitroFlow, parameterName=NiFlow__Tex_Msk3, texture=&T_BlacksRGB}) {name=Tex_Msk3, nodeColor=<rgba 0 1 0.3672 1>}
  NamedRerouteDeclaration(None=@n120) {name=Blend_Contrast, nodeColor=<rgba 0.1016 0 1 1>}
  n132 = Constant3Vector() {constant=<rgba 0 0 -0.25 1>}
  n133 = MaterialFunctionCall(Float4=NamedRerouteUsage()) {materialFunction=&BreakOutFloat4Components}
  n134 = LinearInterpolate(A=Constant3Vector() {constant=<rgba 0 0 0 1>}, B=Constant3Vector() {constant=<rgba 0 0 0 1>}, Alpha=@n133.G)
  NamedRerouteDeclaration(None=LinearInterpolate(A=LinearInterpolate(A=@n132, B=@n134, Alpha=@n133.R), B=Constant3Vector() {constant=<rgba 0.25 0.25 0 1>}, Alpha=@n133.B)) {name=TS_Rotation, nodeColor=<rgba 1 0 0.375 1>}
  n135 = ScalarParameter() {defaultValue=1, group=Wrap_NitroFlow_Spots, parameterName=NiFlow_Spots_Scale_Vehicle, sortPriority=150}
  n136 = ScalarParameter() {defaultValue=1, group=Wrap_NitroFlow_Spots, parameterName=NiFlow_Spots_Scale_Vehicle_Side, sortPriority=150}
  n137 = MaterialFunctionCall(Float4=NamedRerouteUsage()) {materialFunction=&BreakOutFloat4Components}
  n138 = LinearInterpolate(
   A=LinearInterpolate(
    A=ScalarParameter() {defaultValue=2, group=Wrap_NitroFlow_Spots, parameterName=NiFlow_Spots_Scale_Weapon, sortPriority=150}
    B=LinearInterpolate(A=@n135, B=@n136, Alpha=@n137.G)
    Alpha=@n137.R
   )
   B=LinearInterpolate(
    A=ScalarParameter() {defaultValue=1, group=Wrap_NitroFlow_Spots, parameterName=NiFlow_Spots_Scale_Mech, sortPriority=150}
    B=ScalarParameter() {defaultValue=1, group=Wrap_NitroFlow_Spots, parameterName=NiFlow_Spots_Scale_Character, sortPriority=150}
    Alpha=@n137.A
   )
   Alpha=@n137.B
  )
  n139 = ScalarParameter() {defaultValue=1, group=Wrap_NitroFlow_Spots, parameterName=NiFlow_Spots_Scale_Backpack, sortPriority=150}
  n140 = MaterialFunctionCall(Float2=NamedRerouteUsage()) {materialFunction=&BreakOutFloat2Components}
  NamedRerouteDeclaration(None=LinearInterpolate(
    A=LinearInterpolate(A=@n138, B=@n139, Alpha=@n140.R)
    B=ScalarParameter() {defaultValue=1, group=Wrap_NitroFlow_Spots, parameterName=NiFlow_Spots_Scale_Pickaxe, sortPriority=150}
    Alpha=@n140.G
   )) {name=UVs_SpotsScale, nodeColor=<rgba 0 0.7188 1 1>}
  n141 = ComponentMask(None=Add(A=NamedRerouteUsage()) {constB=-0.5}) {g=true}
  NamedRerouteDeclaration(None=Saturate(None=Divide(A=@n141) {constB=-0.001})) {name=Character mask, nodeColor=<rgba 0.9453 1 0 1>}
  n142 = NamedRerouteUsage()
  n143 = NamedRerouteUsage()
  n144 = NamedRerouteUsage()
  n145 = NamedRerouteUsage()
  n146 = NamedRerouteUsage()
  n147 = NamedRerouteUsage()
  n148 = NamedRerouteUsage()
  n149 = Constant() {r=1}
  n150 = NamedRerouteUsage()
  NamedRerouteDeclaration(None=Multiply(
    A=MaterialFunctionCall!(
     Texture_Normals=@n142
     Rotation_Axis=@n143
     Caculated_Bounds=@n144
     Pre_Skinned_Normals=@n145
     IsStatsB=@n146
     IsState=@n147
     Blend_Contrast=@n148
     Fill_alternate_Axis_Opacity=@n149
     Scale=@n150
     Offset_Pivot
     Offset
    ) {materialFunction=&/MF_TopsSideProjection_Normals_V2}.Add_Surface_Normal
    B=ScalarParameter() {defaultValue=0.5, group=Wrap_NitroFlow_Spots, parameterName=NiFlow_Spots_Normal_Intensty}
   )) {name=Spots_Normal, nodeColor=<rgba 0.5938 0 1 1>}
  n151 = NamedRerouteUsage()
  n152 = NamedRerouteUsage()
  n153 = NamedRerouteUsage()
  n154 = NamedRerouteUsage()
  n155 = NamedRerouteUsage()
  n156 = NamedRerouteUsage()
  n157 = Constant() {r=1}
  n158 = VectorParameter() {group=Wrap_NitroFlow_Spots, parameterName=NiFlow_Spots_Alt_Color, defaultValue=<rgba 0 0 0 1>}
  n159 = AppendVector(A=@n158.RGB, B=@n158.RGB)
  n160 = NamedRerouteUsage()
  n161 = NamedRerouteUsage()
  NamedRerouteDeclaration(None=ChannelMaskParameter(None=MaterialFunctionCall!(
     Texture=@n151
     Caculated_Bounds=@n152
     Pre_Skinned_Normals=@n153
     Blend_Contrast=@n154
     IsStatsB=@n155
     IsState=@n156
     Fill_alternate_Axis_Opacity=@n157
     "Fill_alternate_Axis_Color_A=Roughness"=@n159
     Offset
     Scale=@n160
     Offset_Pivot
     Rotation_Axis=@n161
    ) {materialFunction=&/MF_TopsSideProjection_Base_V2}.Combined) {group=Wrap_NitroFlow_Spots, parameterName=NiFlow_Spots_Channel}.R) {name=Spots_Msk_Ch, nodeColor=<rgba 0.7109 1 0 1>}
  NamedRerouteDeclaration(None=@n74) {name=Spots_hieghtlerp, nodeColor=<rgba 0 0.25 1 1>}
  NamedRerouteDeclaration(None=FunctionInput(Preview=Constant4Vector() {constant=<rgba 1 0 0 0>}.RGBA) {bUsePreviewValueAsDefault=true, inputName=IsState, inputType=FunctionInput_Vector4, sortPriority=4}) {name=IsState, nodeColor=<rgba 0.1497 1 0.1633 1>}
  NamedRerouteDeclaration(None=FunctionInput(Preview=Constant2Vector()) {bUsePreviewValueAsDefault=true, inputName=IsState_B, inputType=FunctionInput_Vector2, sortPriority=4}) {name=IsState_B, nodeColor=<rgba 0.1497 1 0.1633 1>}
  n162 = PreSkinnedLocalBounds()
  n163 = Subtract(A=VertexInterpolator(VS=PreSkinnedPosition()).PS, B=@n162.Min)
  n164 = MaterialFunctionCall(Color=Subtract(A=@n162.Max, B=@n162.Max)) {materialFunction=&SplitComponents}
  n165 = Max(A=Max(A=@n164.R, B=@n164.R), B=@n164.B)
  NamedRerouteDeclaration(None=FunctionInput(Preview=Divide(A=@n163, B=@n165)) {bUsePreviewValueAsDefault=true, inputName=Calcated_Bounds, sortPriority=11}) {name=Caclcated_Bounds, nodeColor=<rgba 0 0.9297 1 1>}
  NamedRerouteDeclaration(None=TextureObjectParameter() {group=Wrap_NitroFlow, parameterName=NiFlow__Tex_Msk, texture=&T_BlacksRGB}) {name=Tex_Msk, nodeColor=<rgba 0 1 0.3672 1>}
  n166 = PreSkinnedNormal()
  NamedRerouteDeclaration(None=FunctionInput(Preview=VertexInterpolator(VS=@n166).PS) {bUsePreviewValueAsDefault=true, inputName=Pre-Skinned Local Normal, sortPriority=10}) {name=PreSkin_Normal, nodeColor=<rgba 0.8281 1 0 1>}
  NamedRerouteDeclaration(None=@n58.Normal) {name=Root_Normal, nodeColor=<rgba 0.4062 0 1 1>}
  NamedRerouteUsage()
  NamedRerouteUsage()
  NamedRerouteUsage()
  ScalarParameter() {defaultValue=-9999999, group=Wrap_NitroFlow_Reactive, parameterName=NiFlow_Brightness_Time_BP, sortPriority=100}
 ]}
