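# Material function for the Oak weapon wrap, adding a stylized crosshatched, shadow-banded look to weapons, vehicles and characters behind a static switch.
# Base color blends tint colors through customization masks, darkened by Fresnel outlines and pre-skinned crosshatching.
# Emissive multiplies that color by rounded sun lighting, broken up by a pattern on a camera-facing, FOV-scaled virtual plane, and time-of-day intensity.
# Several branches, including texture edge detection and height lerps, are unconnected.
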MaterialFunction /Game/Weapons/WeaponSkins/Wraps/MaterialFunctions/MF_OakWrap
graph {nodes [
  n1 = CollectionParameter() {
   collection=&FortniteMaterialParameters
   parameterName=SunlightDirectionalVector
  }
  n2 = ComponentMask(None=ComponentMask(None=@n1) {b=true, g=true, r=true}) {b=true, g=true, r=true}
  n3 = ComponentMask(None=PixelNormalWS()) {b=true, g=true, r=true}
  n4 = VertexNormalWS()
  n5 = ShadingPathSwitch(
   Default=LinearInterpolate(
    A=@n3
    B=ComponentMask(None=@n4) {b=true, g=true, r=true}
    Alpha=ScalarParameter() {defaultValue=0.25, group=BarbequeLarry, parameterName=BL_Pixel-Vertex}
   )
   Forward=@n3
   Mobile=@n3
  )
  n6 = DotProduct(A=@n2, B=@n5)
  n7 = Constant3Vector() {constant=<rgba 1 1 0 0>}
  n8 = MaterialFunctionCall() {materialFunction=&CameraDirectionVector}
  n9 = Multiply(A=Constant() {desc=Needs at least a small amount of this or it will break, r=0.25}, B=Normalize(VectorInput=Multiply(A=@n7, B=@n8.Result)))
  n10 = Multiply(A=@n8.Result, B=ScalarParameter() {defaultValue=70, group=BarbequeLarry, parameterName=BL_VP-CameraVectorMin})
  n11 = StaticSwitchParameter(
   True=ObjectPositionWS()
   False=Add(A=ActorPositionWS(), B=VectorParameter() {group=BarbequeLarry, parameterName=BL_VP-ActorPosOffset, defaultValue=<rgba 0 0 80 1>}.RGB)
  ) {group=BarbequeLarry, parameterName=BL_UseObjectPos}
  n12 = FunctionInput() {inputName=IsVehicle, inputType=FunctionInput_Scalar, sortPriority=4}
  n13 = FunctionInput() {inputName=IsMech, inputType=FunctionInput_Scalar, sortPriority=4}
  n14 = ComponentMask(None=Saturate(None=Add(A=@n12, B=@n13))) {r=true}
  n15 = FunctionInput() {inputName=IsCharacter, inputType=FunctionInput_Scalar, sortPriority=4}
  n16 = LinearInterpolate(
   A=LinearInterpolate(
    A=ScalarParameter() {defaultValue=1.4, group=BarbequeLarry, parameterName=BL_VP-Scale-LowFOV}
    B=ScalarParameter() {defaultValue=4.2, group=BarbequeLarry, parameterName=BL_VP-Scale-LowFOV_Vehicle}
    Alpha=@n14
   )
   B=ScalarParameter() {defaultValue=1.4, group=BarbequeLarry, parameterName=BL_VP-Scale-LowFOV_Character}
   Alpha=@n15
  )
  n17 = LinearInterpolate(
   A=LinearInterpolate(
    A=ScalarParameter() {defaultValue=1.75, group=BarbequeLarry, parameterName=BL_VP-Scale-HighFOV}
    B=ScalarParameter() {defaultValue=7, group=BarbequeLarry, parameterName=BL_VP-Scale-HighFOV_Vehicle}
    Alpha=@n14
   )
   B=ScalarParameter() {defaultValue=3, group=BarbequeLarry, parameterName=BL_VP-Scale-HighFOV_Character}
   Alpha=@n15
  )
  n18 = Saturate(None=MaterialFunctionCall(
    Input=MaterialFunctionCall() {materialFunction=/Engine/Functions/Engine_MaterialFunctions02/Utility/FOV}."FOV "
    Input Low=Constant() {r=0.414}
    Input High=Constant() {r=1.091}
    Target Low=Constant()
    Target High=Constant() {r=1}
   ) {materialFunction=&RemapValueRange}.Result)
  n19 = TextureObjectParameter() {
   group=BarbequeLarry
   parameterName=BL_Pattern
   samplerType=SAMPLERTYPE_Masks
   texture=/Game/Global/Textures/Default/Tileable_Patterns/OakPP_Post
  }
  n20 = TextureSample(
   UVs=MaterialFunctionCall(Plane Normal Axis=Add(A=@n9, B=@n8.Result), Plane Center=Add(A=@n10, B=@n11), UV Scale=LinearInterpolate(A=@n16, B=@n17, Alpha=@n18)) {materialFunction=&VirtualPlaneCoordinates}.UVs
   Tex=@n19
  ) {
   samplerSource=SSM_Wrap_WorldGroupSettings
   samplerType=SAMPLERTYPE_LinearGrayscale
   texture=/Game/Characters/Player/Female/Medium/Bodies/F_MED_TNTina/Materials/HalftoneOffset
  }
  n21 = Subtract(
   A=ConstantBiasScale(None=@n6)
   B=Multiply(A=@n20.R, B=ScalarParameter() {defaultValue=0.5, group=BarbequeLarry, parameterName=BL_MidToneSPread})
  )
  n22 = Add(A=Clamp(None=Round(None=@n21)), B=ScalarParameter() {defaultValue=0.2, group=BarbequeLarry, parameterName=BL__DarksAmount})
  n23 = Subtract(
   A=FunctionInput(Preview=Constant() {r=1}) {bUsePreviewValueAsDefault=true, inputName=AO, inputType=FunctionInput_Scalar}
   B=ScalarParameter() {defaultValue=2, group=BarbequeLarry, parameterName=BL_MaskAdjust}
  )
  n24 = ScalarParameter() {defaultValue=2, group=BarbequeLarry, parameterName=BL_MaskMultiplier}
  n25 = Clamp(None=Power(Base=Multiply(A=@n23, B=@n24)))
  Add(
   A=@n22
   B=Multiply(A=ScalarParameter() {defaultValue=0.4, group=BarbequeLarry, parameterName=BL_MaskToBrightenAmount}, B=OneMinus(None=@n25))
  )
  n26 = LinearInterpolate(
   A=LinearInterpolate(
    A=ScalarParameter() {defaultValue=32.0426, group=BarbequeLarry, parameterName=BL_HasMarkSizeWeapon}
    B=ScalarParameter() {defaultValue=148.1746, group=BarbequeLarry, parameterName=BL_HasMarkSizeVehicle}
    Alpha=@n14
   )
   B=ScalarParameter() {defaultValue=64, group=BarbequeLarry, parameterName=BL_HasMarkSizeCharacter}
   Alpha=@n15
  )
  n27 = MaterialFunctionCall() {materialFunction=/Engine/Functions/Engine_MaterialFunctions02/WorldPositionOffset/ObjectScale}
  Divide(A=@n26, B=@n27.Scale XYZ)
  n28 = FunctionInput() {inputName=Material, inputType=FunctionInput_MaterialAttributes, sortPriority=-200}
  n29 = GetMaterialAttributes(None=@n28) {attributeGetTypes [#1, #2, #3]}
  n30 = VectorParameter() {group=BarbequeLarry, parameterName=BL_SecondaryColor, defaultValue=<rgba 0.0142 0.055 0.1354 1>}
  n31 = VectorParameter() {group=BarbequeLarry, parameterName=BL_PrimaryColor, defaultValue=<rgba 0.4427 0 0 1>}
  n32 = FunctionInput() {inputName="Customization Mask (_CM B)", inputType=FunctionInput_Scalar, sortPriority=24, previewValue=<1 0 0 1>}
  n33 = Add(A=ScalarParameter() {defaultValue=1, group=BarbequeLarry, parameterName="BL_UseSecondaryColor_with_BaseColor  1=Disable 0=Active"}, B=@n32)
  n34 = StaticSwitchParameter(True=LinearInterpolate(A=@n30.RGB, B=@n31.RGB, Alpha=@n32), False=LinearInterpolate(A=@n30.RGB, B=@n29.BaseColor, Alpha=Saturate(None=@n33))) {group=BarbequeLarry, parameterName=UseSimpleBaseColor}
  n35 = LinearInterpolate(
   A=@n34
   B=VectorParameter() {group=BarbequeLarry, parameterName=BlendColor, defaultValue=<rgba 0.7188 0.5978 0.4384 1>}.RGB
   Alpha=ScalarParameter() {defaultValue=0.25, group=BarbequeLarry, parameterName=BrigthenDiffuse}
  )
  n36 = ScalarParameter() {defaultValue=-0.391, group=BarbequeLarry, parameterName=Saturation}
  n37 = ScalarParameter() {defaultValue=-5.4358, group=BarbequeLarry, parameterName=BL_OutlineC1}
  n38 = ScalarParameter() {defaultValue=41.451, group=BarbequeLarry, parameterName=BL_OutlineC2}
  n39 = ScalarParameter() {defaultValue=4.7243, group=BarbequeLarry, parameterName=BL_OutlineExponent1}
  n40 = Transform(None=@n29.Normal)
  n41 = Power(
   Base=OneMinus(None=Fresnel(ExponentIn=@n39, Normal=@n40) {baseReflectFraction=0})
   Exp=ScalarParameter() {defaultValue=4.6385, group=BarbequeLarry, parameterName=BL_OutlineExponent2}
  )
  n42 = PreSkinnedPosition()
  n43 = PreSkinnedNormal()
  n44 = ComponentMask(None=MaterialFunctionCall(TextureObject=@n19, TextureSize=@n26, WorldPosition=VertexInterpolator(VS=@n42).PS, World Space Normal=VertexInterpolator(VS=@n43).PS) {materialFunction=&WorldAlignedTexture}.XYZ Texture) {r=true}
  n45 = LinearInterpolate(
   A=ScalarParameter() {defaultValue=-3.8695, group=BarbequeLarry, parameterName=BL_HasmarkIntensity}
   B=ScalarParameter() {defaultValue=0.9201, group=BarbequeLarry, parameterName=BL_HasmarkIntensity2}
   Alpha=LinearInterpolate(A=@n44, B=OneMinus(None=@n44), Alpha=ScalarParameter() {group=BarbequeLarry, parameterName=BL_FlipCrosshatchTexture})
  )
  n46 = Multiply(
   A=Saturate(None=LinearInterpolate(A=@n37, B=@n38, Alpha=@n41))
   B=StaticSwitchParameter(True=Saturate(None=@n45), False=Constant() {r=1}) {defaultValue=true, group=BarbequeLarry, parameterName=UseCrosshatch}
  )
  n47 = FunctionInput() {inputName="Customization Crevasse (_CM G)", inputType=FunctionInput_Scalar, sortPriority=21, previewValue=<1 0 0 1>}
  n48 = LinearInterpolate(
   A=LinearInterpolate(
    A=ScalarParameter() {defaultValue=10, group=BarbequeLarry, parameterName=ZeroMask1}
    B=ScalarParameter() {defaultValue=10, group=BarbequeLarry, parameterName=ZeroMask1_Vehicle}
    Alpha=@n14
   )
   B=LinearInterpolate(
    A=ScalarParameter() {defaultValue=-7, group=BarbequeLarry, parameterName=ZeroMask3}
    B=ScalarParameter() {defaultValue=-7, group=BarbequeLarry, parameterName=ZeroMask3_Vehicle}
    Alpha=@n14
   )
   Alpha=@n47
  )
  n49 = LinearInterpolate(
   A=Constant3Vector()
   B=LinearInterpolate(A=@n34, B=Desaturation(None=@n35, Fraction=@n36), Alpha=@n32)
   Alpha=Multiply(A=Saturate(None=@n46), B=Saturate(None=@n48))
  )
  n50 = Saturate(None=@n22)
  n51 = ScalarParameter() {defaultValue=0.4, group=BarbequeLarry, parameterName=BL_EmissiveAmount-Morning}
  n52 = ScalarParameter() {defaultValue=0.8, group=BarbequeLarry, parameterName=BL_EmissiveAmount-Day}
  n53 = ScalarParameter() {defaultValue=0.3, group=BarbequeLarry, parameterName=BL_EmissiveAmount-Eve}
  n54 = ScalarParameter() {defaultValue=0.3, group=BarbequeLarry, parameterName=BL_EmissiveAmount-Night}
  n55 = StaticSwitchParameter(
   True=Multiply(
    A=Multiply(A=@n49, B=@n50)
    B=MaterialFunctionCall(Value_Morning=@n51, Value_Day=@n52, Value_Evening=@n53, Value_Night=@n54) {materialFunction=&Scalar_Time_of_Day}.Result
   )
   False=@n29.EmissiveColor
  ) {group=BarbequeLarry, parameterName=UseCrosshatchShadows}
  FunctionOutput(None=StaticSwitchParameter(
    True=SetMaterialAttributes!(MaterialAttributes=@n29.MaterialAttributes, BaseColor=@n49, EmissiveColor=@n55) {attributeSetTypes [#1, #2]}
    False=@n28
   ) {group=Customization_Wrap, parameterName=Oak Wrap})
  n56 = CameraVectorWS()
  LinearInterpolate(A=@n2, B=@n56, Alpha=ScalarParameter() {group=BarbequeLarry, parameterName=BL_UseCameraVector})
  n57 = ScalarParameter() {defaultValue=1.25, group=BarbequeLarry, parameterName=BL_LightLevels}
  n58 = Divide(A=Subtract(A=@n6) {constB=-1}, B=@n57)
  n59 = Clamp(
   None=OneMinus(None=Saturate(None=@n58))
   Min=ScalarParameter() {group=BarbequeLarry, parameterName=BL_MinTransitionValue}
   Max=ScalarParameter() {defaultValue=1, group=BarbequeLarry, parameterName=BL_MaxTransitionValue}
  )
  n60 = Saturate(None=@n20.R)
  n61 = ScalarParameter() {defaultValue=1, group=BarbequeLarry, parameterName=BL_HeingtLErpContrast}
  MaterialFunctionCall(Transition Phase=@n59, Height Texture=@n60, Contrast=@n61) {materialFunction=&HeightLerp}
  MaterialFunctionCall!(Diffuse 1, Diffuse 2, Transistion Phase=@n59, Height Texture 1=Constant(), Height Texture 2=@n60, Contrast=@n61) {materialFunction=/Engine/Functions/Engine_MaterialFunctions02/Texturing/HeightLerpWithTwoHeightMaps}
  Multiply(A=@n26, B=@n27.Scale XYZ)
  Multiply(A=@n31.RGB)
  Multiply(A=ObjectRadius(), B=@n26)
  n62 = TextureCoordinate()
  n63 = ScalarParameter() {defaultValue=3, group=BarbequeLarry, parameterName=MIpBLur}
  n64 = TextureSampleParameter2D!(UVs=@n62, Level=@n63, Apply View MipBias) {
   group=BarbequeLarry
   mipValueMode=TMVM_MipLevel
   parameterName=Diffuse
   samplerSource=SSM_Wrap_WorldGroupSettings
   texture=/Game/Weapons/FORT_Shotguns/Textures/T_HammerShotGun_D
  }
  n65 = ScalarParameter() {defaultValue=0, group=BarbequeLarry, parameterName=Offset}
  n66 = MaterialFunctionCall!(OffsetAmount=AppendVector(A=@n65, B=@n65)) {materialFunction=/Engine/Functions/Engine_MaterialFunctions01/ImageAdjustment/BlurSampleOffsets}
  n67 = TextureSampleParameter2D(UVs=Add(A=@n62, B=@n66."1,0"), Level=@n63) {
   group=BarbequeLarry
   mipValueMode=TMVM_MipLevel
   parameterName=Diffuse
   samplerSource=SSM_Wrap_WorldGroupSettings
   texture=&T_HammerShotGun_D
  }
  n68 = TextureSampleParameter2D(UVs=Add(A=TextureCoordinate(), B=@n66."-1,0")) {
   group=BarbequeLarry
   mipValueMode=TMVM_MipLevel
   parameterName=Diffuse
   samplerSource=SSM_Wrap_WorldGroupSettings
   texture=&T_HammerShotGun_D
  }
  n69 = Add(A=Subtract(A=@n64.RGB, B=@n67.RGB), B=Subtract(A=@n64.RGB, B=@n68.RGB))
  n70 = TextureSampleParameter2D(UVs=Add(A=TextureCoordinate(), B=@n66."0,1"), Level=@n63) {
   group=BarbequeLarry
   mipValueMode=TMVM_MipLevel
   parameterName=Diffuse
   samplerSource=SSM_Wrap_WorldGroupSettings
   texture=&T_HammerShotGun_D
  }
  n71 = TextureSampleParameter2D(UVs=Add(A=TextureCoordinate(), B=@n66."0,-1"), Level=@n63) {
   group=BarbequeLarry
   mipValueMode=TMVM_MipLevel
   parameterName=Diffuse
   samplerSource=SSM_Wrap_WorldGroupSettings
   texture=&T_HammerShotGun_D
  }
  n72 = Add(A=Subtract(A=@n64.RGB, B=@n70.RGB), B=Subtract(A=@n64.RGB, B=@n71.RGB))
  n73 = Desaturation(None=Divide(A=Add(A=@n69, B=@n72), B=ScalarParameter() {defaultValue=4, group=BarbequeLarry, parameterName=Samples}))
  n74 = ScalarParameter() {defaultValue=0.0172, group=BarbequeLarry, parameterName=TextureOutlineExp}
  OneMinus(None=ComponentMask(None=Power(Base=@n73, Exp=@n74)) {r=true})
  n75 = Subtract(A=DotProduct(A=@n56, B=@n5)) {constB=-1}
  OneMinus(None=Saturate(None=Divide(A=@n75, B=@n57)))
 ]}
