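# Material function for the OceanRider wrap skin, which applies a two-layer top/side-projected look to weapons, vehicles and mechs.
# MF_WrapBounds feeds projection coordinates, with offset, scale and rotation chosen by the isVehicle and isMech inputs.
# The second layer adds sine-pulsed panning and animated mask-texture warping, blending over the first via projected mask times its alpha.
# Color, roughness, specular and Fresnel-based metallic apply through the customization mask, gated by the Wrap_OceanRider static switch.
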MaterialFunction /Game/Weapons/WeaponSkins/Wraps/MaterialFunctions/MF_OceanRider_Wrap
graph {nodes [
  n1 = FunctionInput() {inputName=Material, inputType=FunctionInput_MaterialAttributes}
  n2 = GetMaterialAttributes(None=@n1) {attributeGetTypes [#1]}
  n3 = TextureObjectParameter() {group=TopSideProjection, parameterName=TS_BaseColor_01, sortPriority=0, texture=&T_BlacksRGB}
  n4 = FunctionInput(Preview=Constant()) {bUsePreviewValueAsDefault=true, inputName=isMech, inputType=FunctionInput_Scalar, sortPriority=4}
  n5 = VectorParameter() {group=TopSideProjection, parameterName="TS_Fill_alternate_Axis_Color_A=Roughness 01", sortPriority=3}
  n6 = VectorParameter() {group=TopSideProjection, parameterName=TS_Fill_Z_Axis_Weapon_Color_Mask_01, sortPriority=21, defaultValue=<rgba 0 0 0 1>}
  n7 = FunctionInput(Preview=Constant()) {inputName=isVehicle, inputType=FunctionInput_Scalar, sortPriority=5}
  n8 = VectorParameter() {group=TopSideProjection_UVS, parameterName=TS_Offset_Weapon_01---------------------, sortPriority=10, defaultValue=<rgba 0 0 0 1>}
  n9 = VectorParameter() {group=TopSideProjection_UVS, parameterName=TS_Offset_Vehicle_01, sortPriority=11, defaultValue=<rgba 0 0 0 1>}
  n10 = VectorParameter() {group=TopSideProjection_UVS, parameterName=TS_Offset_Mech_01, sortPriority=12, defaultValue=<rgba 0 0 0 1>}
  n11 = LinearInterpolate(A=LinearInterpolate(A=@n8.RGB, B=@n9.RGB, Alpha=@n7), B=@n10.RGB, Alpha=@n4)
  n12 = PreSkinnedPosition()
  n13 = MaterialFunctionCall!(
   is Vehicle=@n7
   Object Bounds Max Adjustment=FunctionInput() {bUsePreviewValueAsDefault=true, inputName=Object Bounds Max Adjustment, sortPriority=13}
   Object Bounds Min Adjustment=FunctionInput() {bUsePreviewValueAsDefault=true, inputName=Object Bounds Min Adjustment, sortPriority=12}
   Object Bounds Center Adjustment=FunctionInput() {bUsePreviewValueAsDefault=true, inputName=Object Bounds Center Adjustment, sortPriority=14}
   isMech=@n4
   Side Project Vehicle
   Local Position=FunctionInput(Preview=VertexInterpolator(VS=@n12).PS) {bUsePreviewValueAsDefault=true, inputName=Local Position, sortPriority=11}
  ) {materialFunction=&/MF_WrapBounds}
  n14 = VectorParameter() {group=TopSideProjection_UVS, parameterName="TS_Scale_Weapon_01 A=uniform", sortPriority=30, defaultValue=<rgba 1 1 1 1>}
  n15 = VectorParameter() {group=TopSideProjection_UVS, parameterName="TS_Scale_Vehicle_01 A=uniform", sortPriority=31, defaultValue=<rgba 1 1 1 1>}
  n16 = LinearInterpolate(A=Multiply(A=@n14.RGB, B=@n14.RGB), B=Multiply(A=@n15.RGB, B=@n15.RGB), Alpha=@n7)
  n17 = VectorParameter() {group=TopSideProjection_UVS, parameterName="TS_Scale_Mech_01 A=uniform", defaultValue=<rgba 1 1 1 1>}
  n18 = Multiply(A=@n17.RGB, B=@n17.RGB)
  n19 = VectorParameter() {group=TopSideProjection_UVS, parameterName=TS-Rotation-Weapon-01---------------------, sortPriority=1, defaultValue=<rgba 0 0 -0.25 1>}
  n20 = VectorParameter() {group=TopSideProjection_UVS, parameterName=TS_Rotation_Vehicle_01, sortPriority=2, defaultValue=<rgba 0 0 0 1>}
  n21 = VectorParameter() {group=TopSideProjection_UVS, parameterName=TS_Rotation_Mech_01, sortPriority=3, defaultValue=<rgba 0.25 0.25 0 1>}
  n22 = FunctionInput(Preview=VertexInterpolator(VS=PreSkinnedNormal()).PS) {bUsePreviewValueAsDefault=true, inputName=Pre-Skinned Local Normal, sortPriority=10}
  n23 = MaterialFunctionCall!(
   IsCharacter
   IsStatsB
   Texture=@n3
   Rotation_Axis=LinearInterpolate(A=LinearInterpolate(A=@n19.RGB, B=@n20.RGB, Alpha=@n7), B=@n21.RGB, Alpha=@n4)
   Caculated_Bounds=Multiply(A=Add(A=@n13.Result, B=@n11), B=LinearInterpolate(A=@n16, B=@n18, Alpha=@n4))
   Pre_Skinned_Normals=@n22
   Blend_Contrast=LinearInterpolate(
    A=ScalarParameter() {defaultValue=1, group=TopSideProjection, parameterName=TS_Blend_Contrast, sortPriority=-1}
    B=ScalarParameter() {defaultValue=2, group=TopSideProjection, parameterName=TS_Blend_Contrast_Mech, sortPriority=-1}
    Alpha=@n4
   )
   Fill_alternate_Axis_Opacity=ScalarParameter() {group=TopSideProjection, parameterName=TS_Fill_alternate_Axis_Opacity 01, sortPriority=2}
   "Fill_alternate_Axis_Color_A=Roughness"=AppendVector(A=@n5.RGB, B=@n5.RGB)
   Is_Vehicle=@n7
   Is_Mech=@n4
   Enable_Fill_Z_Weapon=StaticBool() {value=true}
   Fill_Z_Weapon_opacity=ScalarParameter() {group=TopSideProjection, parameterName=TS_Fill_Z_Axis_Weapon_Opacity 01, sortPriority=2}
   Fill_Z_Weapon_Color=AppendVector(A=@n6.RGB, B=@n6.RGB)
  ) {materialFunction=&/MF_TopsSideProjection_Base}
  n24 = TextureObjectParameter() {group=TopSideProjection, parameterName=TS_BaseColor_02, sortPriority=0, texture=&T_BlacksRGB}
  n25 = LinearInterpolate(
   A=ScalarParameter() {defaultValue=1, group=TopSideProjection, parameterName=TS_Blend_Contrast, sortPriority=-1}
   B=ScalarParameter() {defaultValue=2, group=TopSideProjection, parameterName=TS_Blend_Contrast_Mech, sortPriority=-1}
   Alpha=@n4
  )
  n26 = ScalarParameter() {group=TopSideProjection, parameterName=TS_Fill_alternate_Axis_Opacity 02, sortPriority=51}
  n27 = VectorParameter() {group=TopSideProjection, parameterName="TS_Fill_alternate_Axis_Color_A=Roughness 02", sortPriority=52}
  n28 = VectorParameter() {group=TopSideProjection_UVS, parameterName=TS-Rotation-Weapon-02---------------------, sortPriority=150, defaultValue=<rgba 0 0 -0.25 1>}
  n29 = VectorParameter() {group=TopSideProjection_UVS, parameterName=TS_Rotation_Vehicle_02, sortPriority=151, defaultValue=<rgba 0 0 0 1>}
  n30 = VectorParameter() {group=TopSideProjection_UVS, parameterName=TS_Rotation_Mech_02, sortPriority=153, defaultValue=<rgba 0.25 0.25 0 1>}
  n31 = LinearInterpolate(A=LinearInterpolate(A=@n28.RGB, B=@n29.RGB, Alpha=@n7), B=@n30.RGB, Alpha=@n4)
  n32 = VectorParameter() {group=TopSideProjection_UVS, parameterName=TS_Scale_Weapon_02, sortPriority=66, defaultValue=<rgba 1 1 1 1>}
  n33 = VectorParameter() {group=TopSideProjection_UVS, parameterName=TS_Scale_Vehicle_02, sortPriority=67, defaultValue=<rgba 1 1 1 1>}
  n34 = LinearInterpolate(A=Multiply(A=@n32.RGB, B=@n32.RGB), B=Multiply(A=@n33.RGB, B=@n33.RGB), Alpha=@n7)
  n35 = VectorParameter() {group=TopSideProjection_UVS, parameterName=TS_Scale_Mech_02, sortPriority=68, defaultValue=<rgba 1 1 1 1>}
  n36 = Multiply(A=@n35.RGB, B=@n35.RGB)
  n37 = VectorParameter() {group=TopSideProjection_UVS, parameterName=TS_Offset_Weapon_02---------------------, sortPriority=160, defaultValue=<rgba 0 0 0 1>}
  n38 = VectorParameter() {group=TopSideProjection_UVS, parameterName=TS_Offset_Vehicle_02, sortPriority=161, defaultValue=<rgba 0 0 0 1>}
  n39 = VectorParameter() {group=TopSideProjection_UVS, parameterName=TS_Offset_Mech_02, sortPriority=162, defaultValue=<rgba 0 0 0 1>}
  n40 = Add(
   A=Multiply(A=@n13.Result, B=LinearInterpolate(A=@n34, B=@n36, Alpha=@n4))
   B=LinearInterpolate(A=LinearInterpolate(A=@n37.RGB, B=@n38.RGB, Alpha=@n7), B=@n39.RGB, Alpha=@n4)
  )
  n41 = Sine(None=Multiply(A=Time(), B=ScalarParameter() {defaultValue=0.1, group=TopSideProjection_UVS, parameterName=TS_Panning_Pulse_Frequency, sortPriority=173})) {period=1.6}
  n42 = VectorParameter() {group=TopSideProjection_UVS, parameterName=TS_Panning_Weapon_02---------------------, sortPriority=180, defaultValue=<rgba 0 0 0 1>}
  n43 = VectorParameter() {group=TopSideProjection_UVS, parameterName=TS_Panning_Vehicle_02, sortPriority=181, defaultValue=<rgba 0 0 0 1>}
  n44 = VectorParameter() {group=TopSideProjection_UVS, parameterName=TS_Panning_Mech_02, sortPriority=182, defaultValue=<rgba 0 0 0 1>}
  n45 = Multiply(
   A=Clamp(
    None=Power(Base=@n41)
    Min=ScalarParameter() {group=TopSideProjection_UVS, parameterName=TS_Panning_Pulse_Sinewave_Clamp_Min, sortPriority=175}
    Max=ScalarParameter() {defaultValue=1, group=TopSideProjection_UVS, parameterName=TS_Panning_Pulse_Sinewave_Clamp_Max, sortPriority=176}
   )
   B=LinearInterpolate(A=LinearInterpolate(A=@n42.RGB, B=@n43.RGB, Alpha=@n7), B=@n44.RGB, Alpha=@n4)
  )
  n46 = Multiply(
   A=TextureCoordinate()
   B=LinearInterpolate(
    A=ScalarParameter() {defaultValue=1, group=TopSideProjection_UVS, parameterName=TS-WarpUVS-02-TileRate Weapon, sortPriority=203}
    B=ScalarParameter() {defaultValue=1, group=TopSideProjection_UVS, parameterName=TS-WarpUVS-02-TileRate Vehicle, sortPriority=204}
    Alpha=@n7
   )
  )
  n47 = VectorParameter() {group=TopSideProjection_UVS, parameterName=TS-WarpUVS-02-Anitmated Weapon, sortPriority=207, defaultValue=<rgba 0 0 0 1>}
  n48 = VectorParameter() {group=TopSideProjection_UVS, parameterName=TS-WarpUVS-02-Anitmated Vehicle, sortPriority=208, defaultValue=<rgba 0 0 0 1>}
  n49 = Multiply(A=Time() {period=1}, B=ComponentMask(None=LinearInterpolate(A=@n47.RGB, B=@n48.RGB, Alpha=@n7)) {g=true, r=true})
  n50 = TextureObjectParameter() {
   group=TopSideProjection
   parameterName=TS_Mask_02
   samplerType=SAMPLERTYPE_Masks
   sortPriority=1
   texture=&T_BlackMask
  }
  n51 = ChannelMaskParameter(None=TextureSample(UVs=Add(A=@n46, B=@n49), Tex=@n50).RGB) {group=TopSideProjection_UVS, parameterName=TS-WarpUVS-02_Channel, sortPriority=201}
  n52 = Sine(None=Multiply(A=Time(), B=ScalarParameter() {defaultValue=0.1, group=TopSideProjection_UVS, parameterName=TS_Warp02_Pulse_Frequency, sortPriority=209}))
  n53 = ScalarParameter() {defaultValue=0.5, group=TopSideProjection_UVS, parameterName=TS_Warp02_Pulse_Amplitude, sortPriority=210}
  n54 = Clamp(
   None=Add(
    A=Multiply(A=@n52, B=@n53) {constB=0.5}
    B=ScalarParameter() {defaultValue=0.5, group=TopSideProjection_UVS, parameterName=TS_Warp02_Pulse_SineOffset, sortPriority=211}
   )
   Min=ScalarParameter() {group=TopSideProjection_UVS, parameterName=TS_Warp02_Pulse_Sinewave_Clamp_Min, sortPriority=212}
   Max=ScalarParameter() {defaultValue=1, group=TopSideProjection_UVS, parameterName=TS_Warp02_Sinewave_Clamp_Max, sortPriority=213}
  )
  n55 = LinearInterpolate(
   A=Multiply(A=ScalarParameter() {group=TopSideProjection_UVS, parameterName=TS-WarpUVS-02-Amount Weapon, sortPriority=205}, B=@n54)
   B=Multiply(A=ScalarParameter() {group=TopSideProjection_UVS, parameterName=TS-WarpUVS-02-Amount_Vehicle, sortPriority=206}, B=@n54)
   Alpha=@n7
  )
  n56 = MaterialFunctionCall(
   Texture=@n24
   Rotation_Axis=@n31
   Caculated_Bounds=Add(A=Add(A=@n40, B=@n45), B=Multiply(A=@n51.R, B=@n55))
   Pre_Skinned_Normals=@n22
   Blend_Contrast=@n25
   Fill_alternate_Axis_Opacity=@n26
   "Fill_alternate_Axis_Color_A=Roughness"=AppendVector(A=@n27.RGB, B=@n27.RGB)
   Is_Vehicle=@n7
   Is_Mech=@n4
  ) {materialFunction=&MF_TopsSideProjection_Base}
  n57 = VectorParameter() {group=TopSideProjection_UVS, parameterName=TS_MaskOffset_Weapon_02, sortPriority=63, defaultValue=<rgba 0 0 0 1>}
  n58 = VectorParameter() {group=TopSideProjection_UVS, parameterName=TS_MaskOffset_Vehicle_02, sortPriority=64, defaultValue=<rgba 0 0 0 1>}
  n59 = VectorParameter() {group=TopSideProjection_UVS, parameterName=TS_MaskOffset_Mech_02, sortPriority=65, defaultValue=<rgba 0 0 0 1>}
  n60 = Multiply(A=@n13.Result, B=LinearInterpolate(A=LinearInterpolate(A=@n57.RGB, B=@n58.RGB, Alpha=@n7), B=@n59.RGB, Alpha=@n4))
  n61 = VectorParameter() {group=TopSideProjection_UVS, parameterName=TS_Scale_MaskWeapon_02, sortPriority=66, defaultValue=<rgba 1 1 1 1>}
  n62 = VectorParameter() {group=TopSideProjection_UVS, parameterName=TS_Scale_MaskVehicle_02, sortPriority=67, defaultValue=<rgba 1 1 1 1>}
  n63 = VectorParameter() {group=TopSideProjection_UVS, parameterName=TS_Scale_MaskMech_02, sortPriority=68, defaultValue=<rgba 1 1 1 1>}
  n64 = LinearInterpolate(
   A=LinearInterpolate(A=Multiply(A=@n61.RGB, B=@n61.RGB), B=Multiply(A=@n62.RGB, B=@n62.RGB), Alpha=@n7)
   B=Multiply(A=@n63.RGB, B=@n63.RGB)
   Alpha=@n4
  )
  n65 = VectorParameter() {group=TopSideProjection, parameterName=TS_Fill_alternate_Axis_Color_Mask 02, sortPriority=67, defaultValue=<rgba 0 0 0 1>}
  n66 = VectorParameter() {group=TopSideProjection, parameterName=TS_Fill_Z_Axis_Weapon_Color_Mask_02, sortPriority=67, defaultValue=<rgba 0 0 0 1>}
  n67 = ChannelMaskParameter(None=MaterialFunctionCall(
    Texture=@n50
    Rotation_Axis=@n31
    Caculated_Bounds=Multiply(A=@n60, B=@n64)
    Pre_Skinned_Normals=@n22
    Blend_Contrast=@n25
    Fill_alternate_Axis_Opacity=@n26
    "Fill_alternate_Axis_Color_A=Roughness"=AppendVector(A=@n65.RGB, B=@n65.RGB)
    Is_Vehicle=@n7
    Is_Mech=@n4
    Enable_Fill_Z_Weapon=StaticBool() {value=true}
    Fill_Z_Weapon_opacity=ScalarParameter() {group=TopSideProjection, parameterName=TS_Fill_Z_Axis_Weapon_Opacity 02, sortPriority=51}
    Fill_Z_Weapon_Color=AppendVector(A=@n66.RGB, B=@n66.RGB)
   ) {materialFunction=&MF_TopsSideProjection_Base}.Combined) {group=TopSideProjection, parameterName=TS_Mask_OverflowMask_02, sortPriority=66}
  n68 = ScalarParameter() {group=TopSideProjection, parameterName=TS_Roughness_Min_01, sortPriority=13}
  n69 = ScalarParameter() {defaultValue=1, group=TopSideProjection, parameterName=TS_Roughness_Max_01, sortPriority=14}
  n70 = ComponentMask(None=@n23.Combined) {a=true}
  n71 = Fresnel(ExponentIn=ScalarParameter() {group=TopSideProjection, parameterName=TS_Metallic_01, sortPriority=13}, Normal=Transform(None=@n2.Normal))
  n72 = ScalarParameter() {group=TopSideProjection, parameterName=TS_Metallic_01Contrast, sortPriority=13}
  n73 = SetMaterialAttributes!(
   MaterialAttributes=@n2.MaterialAttributes
   BaseColor=LinearInterpolate(
    A=ComponentMask(None=@n23.Combined) {b=true, g=true, r=true}
    B=ComponentMask(None=@n56.Combined) {b=true, g=true, r=true}
    Alpha=Multiply(A=@n67.R, B=@n56.Alpha)
   )
   Roughness=LinearInterpolate(
    A=LinearInterpolate(A=@n68, B=@n69, Alpha=@n70)
    B=ScalarParameter() {defaultValue=0.5, group=TopSideProjection, parameterName=TS_Roughness_02, sortPriority=14}
    Alpha=Multiply(A=@n56.Alpha, B=@n67.R)
   )
   Specular=LinearInterpolate(
    A=ScalarParameter() {defaultValue=0.5, group=TopSideProjection, parameterName=TS_Specularty_01, sortPriority=12}
    B=ScalarParameter() {defaultValue=0.5, group=TopSideProjection, parameterName=TS_Specularty_02, sortPriority=12}
    Alpha=Multiply(A=@n67.R, B=@n56.Alpha)
   )
   Metallic=LinearInterpolate(
    A=MaterialFunctionCall(In=@n71, Contrast=@n72) {materialFunction=&CheapContrast}.Result
    B=ScalarParameter() {group=TopSideProjection, parameterName=TS_Metallic_02, sortPriority=13}
    Alpha=Multiply(A=@n56.Alpha, B=@n67.R)
   )
  ) {attributeSetTypes [
    #2
    #3
    #4
    #5
   ]}
  n74 = Add(
   A=FunctionInput() {bUsePreviewValueAsDefault=true, inputName="Customization Mask (_CM B)", inputType=FunctionInput_Scalar, sortPriority=24}
   B=ScalarParameter() {group=TopSideProjection, parameterName=TS_Remove_Customization Mask, sortPriority=-10}
  )
  n75 = StaticSwitchParameter(True=BlendMaterialAttributes(A=@n2.MaterialAttributes, B=@n73, Alpha=Saturate(None=@n74)), False=@n1) {desc=TopSide Projection Base - Variable blend between two TS using a warped mask, group=Customization_Wrap, parameterName=Wrap_OceanRider}
  FunctionOutput(None=@n75)
  PreSkinnedLocalBounds()
 ]}
