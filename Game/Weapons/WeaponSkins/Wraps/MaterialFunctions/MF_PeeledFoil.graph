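# Material function for the PeeledFoil weapon-skin wrap, which also covers vehicles and mechs and passes input attributes through when its static switch is off.
# Two foil layers use bounds-based top/side projection, with per-type rotation, offset, scale and time-driven panning.
# Layer 01 blends over the input by customization mask, then takes a flat accent.
# A projected Mask_01 channel blends that result against layer 02; Mask_02's projection is unconnected.
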MaterialFunction /Game/Weapons/WeaponSkins/Wraps/MaterialFunctions/MF_PeeledFoil
graph {nodes [
  n1 = VectorParameter() {group=PeeledFoil, parameterName=PF_Fill_Z_Axis_Weapon_Color_Mask_01, sortPriority=21, defaultValue=<rgba 0 0 0 1>}
  AppendVector(A=@n1.RGB, B=@n1.RGB)
  n2 = VectorParameter() {group=PeeledFoil, parameterName=PF_Fill_Z_Axis_Weapon_Color_Mask_02, sortPriority=67, defaultValue=<rgba 0 0 0 1>}
  AppendVector(A=@n2.RGB, B=@n2.RGB)
  n3 = TextureObjectParameter() {group=PeeledFoil, parameterName=PF_BaseColor_02, sortPriority=56, texture=/Game/Weapons/WeaponSkins/Wraps/Textures/UVCheaker}
  n4 = FunctionInput(Preview=Constant()) {bUsePreviewValueAsDefault=true, inputName=isMech, inputType=FunctionInput_Scalar, sortPriority=4}
  n5 = LinearInterpolate(
   A=ScalarParameter() {defaultValue=1, group=PeeledFoil, parameterName=PF_Blend_Contrast, sortPriority=-1}
   B=ScalarParameter() {defaultValue=2, group=PeeledFoil, parameterName=PF_Blend_Contrast_Mech, sortPriority=-1}
   Alpha=@n4
  )
  n6 = ScalarParameter() {group=PeeledFoil, parameterName=PF_Fill_alternate_Axis_Opacity 02, sortPriority=51}
  n7 = VectorParameter() {group=PeeledFoil, parameterName="PF_Fill_alternate_Axis_Color_A=Roughness 02", sortPriority=52}
  n8 = VectorParameter() {group=PeeledFoil_UVS, parameterName=PF-Rotation-Weapon-02---------------------, sortPriority=150, defaultValue=<rgba 0 0 -0.25 1>}
  n9 = VectorParameter() {group=PeeledFoil_UVS, parameterName=PF_Rotation_Vehicle_02, sortPriority=151, defaultValue=<rgba 0 0 0 1>}
  n10 = FunctionInput(Preview=Constant()) {inputName=isVehicle, inputType=FunctionInput_Scalar, sortPriority=5}
  n11 = VectorParameter() {group=PeeledFoil_UVS, parameterName=PF_Rotation_Mech_02, sortPriority=153, defaultValue=<rgba 0.25 0.25 0 1>}
  n12 = LinearInterpolate(A=LinearInterpolate(A=@n8.RGB, B=@n9.RGB, Alpha=@n10), B=@n11.RGB, Alpha=@n4)
  n13 = VectorParameter() {group=PeeledFoil_UVS, parameterName=PF_Offset_Weapon_02---------------------, sortPriority=160, defaultValue=<rgba 0 0 0 1>}
  n14 = VectorParameter() {group=PeeledFoil_UVS, parameterName=PF_Offset_Vehicle_02, sortPriority=161, defaultValue=<rgba 0 0 0 1>}
  n15 = VectorParameter() {group=PeeledFoil_UVS, parameterName=PF_Offset_Mech_02, sortPriority=162, defaultValue=<rgba 0 0 0 1>}
  n16 = LinearInterpolate(A=LinearInterpolate(A=@n13.RGB, B=@n14.RGB, Alpha=@n10), B=@n15.RGB, Alpha=@n4)
  n17 = PreSkinnedPosition()
  n18 = MaterialFunctionCall!(
   is Vehicle=@n10
   Object Bounds Max Adjustment=FunctionInput() {bUsePreviewValueAsDefault=true, inputName=Object Bounds Max Adjustment, sortPriority=13}
   Object Bounds Min Adjustment=FunctionInput() {bUsePreviewValueAsDefault=true, inputName=Object Bounds Min Adjustment, sortPriority=12}
   Object Bounds Center Adjustment=FunctionInput() {bUsePreviewValueAsDefault=true, inputName=Object Bounds Center Adjustment, sortPriority=14}
   isMech=@n4
   Side Project Vehicle
   Local Position=FunctionInput(Preview=VertexInterpolator(VS=@n17).PS) {bUsePreviewValueAsDefault=true, inputName=Local Position, sortPriority=11}
  ) {materialFunction=&/MF_WrapBounds}
  n19 = Time()
  n20 = VectorParameter() {group=PeeledFoil_UVS, parameterName=PF_Panning_Weapon_02---------------------, sortPriority=180, defaultValue=<rgba 0 0 0 1>}
  n21 = VectorParameter() {group=PeeledFoil_UVS, parameterName=PF_Panning_Vehicle_02, sortPriority=181, defaultValue=<rgba 0 0 0 1>}
  n22 = VectorParameter() {group=PeeledFoil_UVS, parameterName=PF_Panning_Mech_02, sortPriority=182, defaultValue=<rgba 0 0 0 1>}
  n23 = LinearInterpolate(A=LinearInterpolate(A=@n20.RGB, B=@n21.RGB, Alpha=@n10), B=@n22.RGB, Alpha=@n4)
  n24 = VectorParameter() {group=PeeledFoil_UVS, parameterName="PF_Scale_Weapon_02 A=uniform", sortPriority=170, defaultValue=<rgba 1 1 1 1>}
  n25 = Multiply(A=@n24.RGB, B=@n24.RGB)
  n26 = VectorParameter() {group=PeeledFoil_UVS, parameterName="PF_Scale_Vehicle_02 A=uniform", sortPriority=171, defaultValue=<rgba 1 1 1 1>}
  n27 = Multiply(A=@n26.RGB, B=@n26.RGB)
  n28 = VectorParameter() {group=PeeledFoil_UVS, parameterName="PF_Scale_Mech_02 A=uniform", sortPriority=172, defaultValue=<rgba 1 1 1 1>}
  n29 = Multiply(
   A=Add(A=Add(A=@n18.Result, B=@n16), B=Multiply(A=@n19, B=@n23))
   B=LinearInterpolate(A=LinearInterpolate(A=@n25, B=@n27, Alpha=@n10), B=Multiply(A=@n28.RGB, B=@n28.RGB), Alpha=@n4)
  )
  n30 = FunctionInput(Preview=VertexInterpolator(VS=PreSkinnedNormal()).PS) {bUsePreviewValueAsDefault=true, inputName=Pre-Skinned Local Normal, sortPriority=10}
  n31 = MaterialFunctionCall!(
   IsCharacter
   IsStatsB
   Texture=@n3
   Rotation_Axis=@n12
   Caculated_Bounds=@n29
   Pre_Skinned_Normals=@n30
   Blend_Contrast=@n5
   Fill_alternate_Axis_Opacity=@n6
   "Fill_alternate_Axis_Color_A=Roughness"=AppendVector(A=@n7.RGB, B=@n7.RGB)
   Is_Vehicle=@n10
   Is_Mech=@n4
   Enable_Fill_Z_Weapon
   Fill_Z_Weapon_opacity
   Fill_Z_Weapon_Color
  ) {materialFunction=&/MF_TopsSideProjection_Base}
  n32 = FunctionInput() {inputName=Material, inputType=FunctionInput_MaterialAttributes}
  n33 = GetMaterialAttributes(None=@n32) {attributeGetTypes [
    #1
    #2
    #3
    #4
    #5
   ]}
  n34 = TextureObjectParameter() {group=PeeledFoil, parameterName=PF_BaseColor_01, sortPriority=11, texture=&UVCheaker}
  n35 = LinearInterpolate(
   A=ScalarParameter() {defaultValue=1, group=PeeledFoil, parameterName=PF_Blend_Contrast, sortPriority=-1}
   B=ScalarParameter() {defaultValue=2, group=PeeledFoil, parameterName=PF_Blend_Contrast_Mech, sortPriority=-1}
   Alpha=@n4
  )
  n36 = ScalarParameter() {group=PeeledFoil, parameterName=PF_Fill_alternate_Axis_Opacity 01, sortPriority=2}
  n37 = VectorParameter() {group=PeeledFoil, parameterName="PF_Fill_alternate_Axis_Color_A=Roughness 01", sortPriority=3}
  n38 = VectorParameter() {group=PeeledFoil_UVS, parameterName=PF_Offset_Weapon_01---------------------, sortPriority=10, defaultValue=<rgba 0 0 0 1>}
  n39 = VectorParameter() {group=PeeledFoil_UVS, parameterName=PF_Offset_Vehicle_01, sortPriority=11, defaultValue=<rgba 0 0 0 1>}
  n40 = VectorParameter() {group=PeeledFoil_UVS, parameterName=PF_Offset_Mech_01, sortPriority=12, defaultValue=<rgba 0 0 0 1>}
  n41 = LinearInterpolate(A=LinearInterpolate(A=@n38.RGB, B=@n39.RGB, Alpha=@n10), B=@n40.RGB, Alpha=@n4)
  n42 = Time()
  n43 = VectorParameter() {group=PeeledFoil_UVS, parameterName=PF_Panning_Weapon_01---------------------, sortPriority=40, defaultValue=<rgba 0 0 0 1>}
  n44 = VectorParameter() {group=PeeledFoil_UVS, parameterName=PF_Panning_Vehicle_01, sortPriority=41, defaultValue=<rgba 0 0 0 1>}
  n45 = VectorParameter() {group=PeeledFoil_UVS, parameterName=PF_Panning_Mech_01, sortPriority=42, defaultValue=<rgba 0 0 0 1>}
  n46 = LinearInterpolate(A=LinearInterpolate(A=@n43.RGB, B=@n44.RGB, Alpha=@n10), B=@n45.RGB, Alpha=@n4)
  n47 = VectorParameter() {group=PeeledFoil_UVS, parameterName="PF_Scale_Weapon_01 A=uniform", sortPriority=30, defaultValue=<rgba 1 1 1 1>}
  n48 = Multiply(A=@n47.RGB, B=@n47.RGB)
  n49 = VectorParameter() {group=PeeledFoil_UVS, parameterName="PF_Scale_Vehicle_01 A=uniform", sortPriority=31, defaultValue=<rgba 1 1 1 1>}
  n50 = Multiply(A=@n49.RGB, B=@n49.RGB)
  n51 = VectorParameter() {group=PeeledFoil_UVS, parameterName="PF_Scale_Mech_01 A=uniform", defaultValue=<rgba 1 1 1 1>}
  n52 = Multiply(
   A=Add(A=Add(A=@n18.Result, B=@n41), B=Multiply(A=@n42, B=@n46))
   B=LinearInterpolate(A=LinearInterpolate(A=@n48, B=@n50, Alpha=@n10), B=Multiply(A=@n51.RGB, B=@n51.RGB), Alpha=@n4)
  )
  n53 = VectorParameter() {group=PeeledFoil_UVS, parameterName=PF-Rotation-Weapon-01---------------------, sortPriority=1, defaultValue=<rgba 0 0 -0.25 1>}
  n54 = VectorParameter() {group=PeeledFoil_UVS, parameterName=PF_Rotation_Vehicle_01, sortPriority=2, defaultValue=<rgba 0 0 0 1>}
  n55 = VectorParameter() {group=PeeledFoil_UVS, parameterName=PF_Rotation_Mech_01, sortPriority=3, defaultValue=<rgba 0.25 0.25 0 1>}
  n56 = LinearInterpolate(A=LinearInterpolate(A=@n53.RGB, B=@n54.RGB, Alpha=@n10), B=@n55.RGB, Alpha=@n4)
  n57 = MaterialFunctionCall(
   Texture=@n34
   Rotation_Axis=@n56
   Caculated_Bounds=@n52
   Pre_Skinned_Normals=@n30
   Blend_Contrast=@n35
   Fill_alternate_Axis_Opacity=@n36
   "Fill_alternate_Axis_Color_A=Roughness"=AppendVector(A=@n37.RGB, B=@n37.RGB)
   Is_Vehicle=@n10
   Is_Mech=@n4
  ) {materialFunction=&MF_TopsSideProjection_Base}
  n58 = GetMaterialAttributes(None=@n32) {attributeGetTypes [#4]}
  n59 = TextureObjectParameter() {
   group=PeeledFoil
   parameterName=PF_Normal_01
   samplerType=SAMPLERTYPE_Normal
   sortPriority=16
   texture=&FlatNormalBlue
  }
  n60 = MaterialFunctionCall!(
   Texture_Normals=@n59
   Rotation_Axis=@n56
   Caculated_Bounds=@n52
   Pre_Skinned_Normals=@n30
   Blend_Contrast=@n35
   Fill_alternate_Axis_Opacity=@n36
   Is_Vehicle=@n10
   Is_Mech=@n4
  ) {materialFunction=&/MF_TopsSideProjection_Normals}
  n61 = TextureObjectParameter() {
   group=PeeledFoil
   parameterName=PF_Mask_01
   samplerType=SAMPLERTYPE_Masks
   sortPriority=20
   texture=&T_BlackMask
  }
  n62 = VectorParameter() {group=PeeledFoil, parameterName=PF_Fill_alternate_Axis_Color_Mask_01, sortPriority=21, defaultValue=<rgba 0 0 0 1>}
  n63 = MaterialFunctionCall(
   Texture=@n61
   Rotation_Axis=@n56
   Caculated_Bounds=@n52
   Pre_Skinned_Normals=@n30
   Blend_Contrast=@n35
   Fill_alternate_Axis_Opacity=@n36
   "Fill_alternate_Axis_Color_A=Roughness"=AppendVector(A=@n62.RGB, B=@n62.RGB)
   Is_Vehicle=@n10
   Is_Mech=@n4
  ) {materialFunction=&MF_TopsSideProjection_Base}
  n64 = MaterialFunctionCall!(
   A_01=@n33.BaseColor
   B_01=ComponentMask(None=@n57.Combined) {b=true, g=true, r=true}
   A_02=@n33.BaseColor
   B_02=LinearInterpolate(
    A=ScalarParameter() {group=PeeledFoil, parameterName=PF_Roughness_Min_01, sortPriority=13}
    B=ScalarParameter() {defaultValue=1, group=PeeledFoil, parameterName=PF_Roughness_Max_01, sortPriority=14}
    Alpha=ComponentMask(None=@n57.Combined) {a=true}
   )
   A_03=@n33.BaseColor
   B_03=ChannelMaskParameter(None=@n63.Combined) {group=PeeledFoil, parameterName=PF_MetalicChannel_Mask_01, sortPriority=26}.R
   A_04=@n33.BaseColor
   B_04=LinearInterpolate(
    A=@n58.Normal
    B=Add(A=@n58.Normal, B=@n60.Add_Surface_Normal)
    Alpha=ScalarParameter() {defaultValue=1, group=PeeledFoil, parameterName=PF_Normal_Intensty_01, sortPriority=1}
   )
   A_05=@n33.BaseColor
   B_05=ScalarParameter() {defaultValue=0.5, group=PeeledFoil, parameterName=PF_Specular_01, sortPriority=13}
   Alpha=FunctionInput() {bUsePreviewValueAsDefault=true, inputName="Customization Mask (_CM B)", inputType=FunctionInput_Scalar, sortPriority=24}
  ) {materialFunction=/Game/Packages/Fortress_SharedMaterials/Base_Material_Functions/BaseCustomizationFunctions/MF_Blend5}
  n65 = VectorParameter() {group=PeeledFoil, parameterName=PF_Accent_Simple_Color, sortPriority=400, defaultValue=<rgba 0 0 0 1>}
  n66 = MaterialFunctionCall(
   A_01=@n64.Out_01
   B_01=@n65.RGB
   A_02=@n64.Out_01
   B_02=ScalarParameter() {defaultValue=0.5, group=PeeledFoil, parameterName=PF_Accent_Roughness, sliderMax=1, sortPriority=401}
   A_03=@n64.Out_01
   B_03=ScalarParameter() {group=PeeledFoil, parameterName=PF_Accent_Metalness, sliderMax=1, sortPriority=401}
   A_04=@n64.Out_01
   B_04=@n33.Normal
   A_05=@n64.Out_01
   B_05=ScalarParameter() {defaultValue=0.5, group=PeeledFoil, parameterName=PF_Accent_Specular, sliderMax=1, sortPriority=401}
   Alpha=FunctionInput() {bUsePreviewValueAsDefault=true, inputName=Customization Mask Accent, inputType=FunctionInput_Scalar, sortPriority=24}
  ) {materialFunction=&MF_Blend5}
  n67 = TextureObjectParameter() {
   group=PeeledFoil
   parameterName=PF_Normal_02
   samplerType=SAMPLERTYPE_Normal
   sortPriority=61
   texture=&FlatNormalBlue
  }
  n68 = MaterialFunctionCall(
   Texture_Normals=@n67
   Rotation_Axis=@n12
   Caculated_Bounds=@n29
   Pre_Skinned_Normals=@n30
   Blend_Contrast=@n5
   Fill_alternate_Axis_Opacity=@n6
   Is_Vehicle=@n10
   Is_Mech=@n4
  ) {materialFunction=&MF_TopsSideProjection_Normals}
  n69 = MaterialFunctionCall(
   A_01=ComponentMask(None=@n31.Combined) {b=true, g=true, r=true}
   B_01=@n66.Out_01
   A_02=LinearInterpolate(
    A=ScalarParameter() {group=PeeledFoil, parameterName=PF_Roughness_Min 02, sortPriority=57}
    B=ScalarParameter() {defaultValue=1, group=PeeledFoil, parameterName=PF_Roughness_Max_02, sortPriority=58}
    Alpha=ComponentMask(None=@n31.Combined) {a=true}
   )
   B_02=@n66.Out_01
   A_03=ScalarParameter() {group=PeeledFoil, parameterName=PF_Metalic_02, sortPriority=66}
   B_03=@n66.Out_01
   A_04=LinearInterpolate(
    A=@n58.Normal
    B=Add(A=@n58.Normal, B=@n68.Add_Surface_Normal)
    Alpha=ScalarParameter() {defaultValue=1, group=PeeledFoil, parameterName=PF_Normal_Intensty_02, sortPriority=61}
   )
   B_04=@n66.Out_01
   A_05=ScalarParameter() {defaultValue=0.5, group=PeeledFoil, parameterName=PF_Specular_Min 02, sortPriority=57}
   B_05=@n66.Out_01
   Alpha=ChannelMaskParameter(None=@n63.Combined) {group=PeeledFoil, parameterName=PF_Blend_Mask_01, sortPriority=72}.R
  ) {materialFunction=&MF_Blend5}
  n70 = SetMaterialAttributes!(
   MaterialAttributes=@n58.MaterialAttributes
   BaseColor=@n69.Out_01
   Roughness=@n69.Out_01
   Metallic=@n69.Out_01
   Normal=@n69.Out_01
   Specular=@n69.Out_01
  ) {attributeSetTypes [
    #1
    #2
    #3
    #4
    #5
   ]}
  FunctionOutput(None=StaticSwitchParameter(True=@n70, False=@n32) {group=Customization_Wrap, parameterName=PeeledFoil})
  n71 = TextureObjectParameter() {
   group=PeeledFoil
   parameterName=PF_Mask_02
   samplerType=SAMPLERTYPE_Masks
   sortPriority=62
   texture=&T_BlackMask
  }
  n72 = VectorParameter() {group=PeeledFoil, parameterName=PF_Fill_alternate_Axis_Color_Mask 02, sortPriority=67, defaultValue=<rgba 0 0 0 1>}
  MaterialFunctionCall(
   Texture=@n71
   Rotation_Axis=@n12
   Caculated_Bounds=@n29
   Pre_Skinned_Normals=@n30
   Blend_Contrast=@n5
   Fill_alternate_Axis_Opacity=@n6
   "Fill_alternate_Axis_Color_A=Roughness"=AppendVector(A=@n72.RGB, B=@n72.RGB)
   Is_Vehicle=@n10
   Is_Mech=@n4
  ) {materialFunction=&MF_TopsSideProjection_Base}
  ScalarParameter() {group=PeeledFoil, parameterName=PF_Fill_Z_Axis_Weapon_Opacity 01, sortPriority=2}
  ScalarParameter() {group=PeeledFoil, parameterName=PF_Fill_Z_Axis_Weapon_Opacity 02, sortPriority=51}
  StaticBool() {value=true}
 ]}
