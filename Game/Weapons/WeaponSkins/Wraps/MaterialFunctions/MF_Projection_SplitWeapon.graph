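# Material function for weapon wrap skins that projects a texture layer onto the incoming material, toggled by the Wrap_Projecton_SplitWeapon static switch.
# Bounds-based UVs drive base color, while MF_TopsSideProjection_Base feeds roughness.
# IsState inputs blend scale, offset and rotation parameters between weapon, vehicle, mech, character, backpack and pickaxe.
# Primary plus scaled secondary masks control the output blend; a further accent blend is left unconnected.
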MaterialFunction /Game/Weapons/WeaponSkins/Wraps/MaterialFunctions/MF_Projection_SplitWeapon
graph {nodes [
  n1 = TextureSample() {samplerType=SAMPLERTYPE_Masks, texture=/Game/Weapons/FORT_Rifles/Textures/T_Scar_CM}
  n2 = FunctionInput(Preview=@n1.R) {bUsePreviewValueAsDefault=true, inputName=Mask Primary, inputType=FunctionInput_Scalar, sortPriority=24, previewValue=<1 0 0 1>}
  n3 = FunctionInput(Preview=@n1.B) {bUsePreviewValueAsDefault=true, inputName=Mask Secondary, inputType=FunctionInput_Scalar, sortPriority=24, previewValue=<1 0 0 1>}
  n4 = Multiply(A=@n3, B=ScalarParameter() {group=Projection_SplitWeapon, parameterName="SplitW_Secondary_msk 0=On 1=Off", sortPriority=10}) {constB=0}
  n5 = BlendMaterialAttributes(A=NamedRerouteUsage(), B=NamedRerouteUsage(), Alpha=Saturate(None=Add(A=@n2, B=@n4))) {vertexAttributeBlendType=UseA}
  BlendMaterialAttributes(A=@n5, B=NamedRerouteUsage(), Alpha=NamedRerouteDeclaration(None=@n3) {name=Msk_Accent, nodeColor=<rgba 0 1 0.7656 1>}) {vertexAttributeBlendType=UseA}
  n6 = MaterialFunctionCall(Float4=NamedRerouteUsage()) {materialFunction=&BreakOutFloat4Components}
  n7 = VectorParameter() {group=Projection_SplitWeapon, parameterName=SplitW_Fill_Alterante_Aixs Color, defaultValue=<rgba 0 0 0 1>}
  n8 = MaterialFunctionCall!(
   IsCharacter=@n6.A
   IsStatsB=NamedRerouteUsage()
   Texture=NamedRerouteUsage()
   Rotation_Axis=NamedRerouteUsage()
   Caculated_Bounds=Multiply(A=NamedRerouteUsage(), B=NamedRerouteUsage())
   Pre_Skinned_Normals=NamedRerouteUsage()
   Blend_Contrast=ScalarParameter() {group=Projection_SplitWeapon, parameterName=SplitW_Contrast}
   Fill_alternate_Axis_Opacity=ScalarParameter() {group=Projection_SplitWeapon, parameterName=SplitW_Fill_Alternate_Axis_Opacity}
   "Fill_alternate_Axis_Color_A=Roughness"=AppendVector(A=@n7.RGB, B=@n7.RGB)
   Is_Vehicle=@n6.A
   Is_Mech=@n6.A
   Enable_Fill_Z_Weapon
   Fill_Z_Weapon_opacity
   Fill_Z_Weapon_Color
  ) {materialFunction=&/MF_TopsSideProjection_Base}
  ChannelMaskParameter(None=@n8.Combined) {group=Projection_SplitWeapon, parameterName=SplitW_Tex_Msk_Rougness_Channel}
  n9 = TextureSample() {samplerType=SAMPLERTYPE_Normal, texture=/Game/Weapons/FORT_Rifles/T_SCAR_N}
  n10 = Constant3Vector() {constant=<rgba 1 0 0 0>}
  n11 = GetMaterialAttributes(None=FunctionInput(Preview=SetMaterialAttributes!(MaterialAttributes, Normal=@n9.RGB, BaseColor=@n10) {attributeSetTypes [#1, #2]}) {inputName=Material, inputType=FunctionInput_MaterialAttributes}) {attributeGetTypes [
    #2
    #3
    #1
    #4
   ]}
  FunctionOutput(None=StaticSwitchParameter(True=@n5, False=NamedRerouteDeclaration(None=@n11.MaterialAttributes) {name=Root_Material, nodeColor=<rgba 0.7812 1 0 1>}) {group=Customization, parameterName=Wrap_Projecton_SplitWeapon})
  n12 = Add(A=Multiply(A=NamedRerouteUsage(), B=Constant3Vector() {constant=<rgba 1 1 -1 0>}), B=Constant3Vector() {constant=<rgba 0 0 1 0>})
  n13 = NamedRerouteUsage()
  n14 = MaterialFunctionCall(Color=Add(A=Multiply(A=@n12, B=@n13), B=NamedRerouteUsage()) {constB=0}) {materialFunction=&SplitComponents}
  n15 = AppendVector(A=@n14.B, B=@n14.B)
  n16 = MaterialFunctionCall(Float4=NamedRerouteUsage()) {materialFunction=&BreakOutFloat4Components}
  LinearInterpolate(B=@n15, Alpha=@n16.G)
  n17 = AppendVector(A=@n14.G, B=OneMinus(None=@n14.R))
  n18 = AppendVector(A=@n14.R, B=@n14.R)
  n19 = TextureSample(UVs=LinearInterpolate(A=LinearInterpolate(A=@n17, B=@n18, Alpha=@n16.R), B=OneMinus(None=@n15), Alpha=@n16.B), Tex=NamedRerouteUsage())
  Multiply(A=@n19.RGB)
  n20 = PreSkinnedLocalBounds()
  n21 = Subtract(A=VertexInterpolator(VS=PreSkinnedPosition()).PS, B=@n20.Min)
  n22 = MaterialFunctionCall(Color=Subtract(A=@n20.Max, B=@n20.Max)) {materialFunction=&SplitComponents}
  n23 = Max(A=Max(A=@n22.R, B=@n22.R), B=@n22.B)
  NamedRerouteDeclaration(None=FunctionInput(Preview=Divide(A=@n21, B=@n23)) {bUsePreviewValueAsDefault=true, inputName=Calcated_Bounds, sortPriority=11}) {name=Caclcated_Bounds, nodeColor=<rgba 0 0.9297 1 1>}
  n24 = PreSkinnedNormal()
  NamedRerouteDeclaration(None=FunctionInput(Preview=VertexInterpolator(VS=@n24).PS) {bUsePreviewValueAsDefault=true, inputName=Pre-Skinned Local Normal, sortPriority=10}) {name=PreSkin_Normal, nodeColor=<rgba 0.8281 1 0 1>}
  NamedRerouteDeclaration(None=TextureObjectParameter() {group=Projection_SplitWeapon, parameterName=SplitW__Tex_BC, texture=/Game/Weapons/WeaponSkins/Wraps/Textures/UVCheaker}) {name=Tex_BC, nodeColor=<rgba 1 0 0.0703 1>}
  NamedRerouteDeclaration(None=TextureObjectParameter() {group=Projection_SplitWeapon, parameterName=SplitW__Tex_Msk, texture=&T_BlacksRGB}) {name=Tex_Msk, nodeColor=<rgba 1 0 0.0703 1>}
  n25 = ScalarParameter() {defaultValue=1, group=Projection_SplitWeapon_Uvs, parameterName=SplitW_Scale_Vehicle, sortPriority=150}
  n26 = ScalarParameter() {defaultValue=1, group=Projection_SplitWeapon_Uvs, parameterName=SplitW_Scale_Vehicle_Side, sortPriority=150}
  n27 = MaterialFunctionCall(Float4=NamedRerouteUsage()) {materialFunction=&BreakOutFloat4Components}
  n28 = LinearInterpolate(
   A=LinearInterpolate(
    A=ScalarParameter() {defaultValue=1, group=Projection_SplitWeapon_Uvs, parameterName=SplitW_Scale_Weapon, sortPriority=150}
    B=LinearInterpolate(A=@n25, B=@n26, Alpha=@n27.G)
    Alpha=@n27.R
   )
   B=LinearInterpolate(
    A=ScalarParameter() {defaultValue=1, group=Projection_SplitWeapon_Uvs, parameterName=SplitW_Scale_Mech, sortPriority=150}
    B=ScalarParameter() {defaultValue=1, group=Projection_SplitWeapon_Uvs, parameterName=SplitW_Scale_Character, sortPriority=150}
    Alpha=@n27.A
   )
   Alpha=@n27.B
  )
  n29 = ScalarParameter() {defaultValue=1, group=Projection_SplitWeapon_Uvs, parameterName=SplitW_Scale_Backpack, sortPriority=150}
  n30 = MaterialFunctionCall(Float2=NamedRerouteUsage()) {materialFunction=&BreakOutFloat2Components}
  NamedRerouteDeclaration(None=LinearInterpolate(
    A=LinearInterpolate(A=@n28, B=@n29, Alpha=@n30.R)
    B=ScalarParameter() {defaultValue=1, group=Projection_SplitWeapon_Uvs, parameterName=SplitW_Scale_Pickaxe, sortPriority=150}
    Alpha=@n30.G
   )) {name=UVs_Scale_BC, nodeColor=<rgba 0 0.7188 1 1>}
  n31 = SetMaterialAttributes!(
   MaterialAttributes=NamedRerouteUsage()
   BaseColor=@n19.RGB
   Roughness=LinearInterpolate(
    A=ScalarParameter() {group=Projection_SplitWeapon, parameterName=SplitW_Rougness_High, sortPriority=10}
    B=ScalarParameter() {defaultValue=1, group=Projection_SplitWeapon, parameterName=SplitW_Rougness_Low, sortPriority=10}
    Alpha=@n8.Alpha
   )
   Specular=ScalarParameter() {defaultValue=0.5, group=Projection_SplitWeapon, parameterName=SplitW_Spec, sortPriority=10}
   Metallic=ScalarParameter() {group=Projection_SplitWeapon, parameterName=SplitW_Metalic, sortPriority=10}
   Normal
   EmissiveColor
  ) {attributeSetTypes [
    #2
    #3
    #5
    #4
    #1
    #6
   ]}
  NamedRerouteDeclaration(None=@n31) {name=Mat_Base, nodeColor=<rgba 0 0.3906 1 1>}
  n32 = VectorParameter() {group=Projection_SplitWeapon_Uvs, parameterName=SplitW_Offset_Vehicle, defaultValue=<rgba 0 0 0 1>}
  n33 = VectorParameter() {group=Projection_SplitWeapon_Uvs, parameterName=SplitW_Offset_VehicleSide, defaultValue=<rgba 0 0 0 1>}
  n34 = MaterialFunctionCall(Float4=NamedRerouteUsage()) {materialFunction=&BreakOutFloat4Components}
  n35 = LinearInterpolate(
   A=LinearInterpolate(
    A=VectorParameter() {group=Projection_SplitWeapon_Uvs, parameterName=SplitW_Offset_Weapon, defaultValue=<rgba 0 0 0 1>}.RGB
    B=LinearInterpolate(A=@n32.RGB, B=@n33.RGB, Alpha=@n34.G)
    Alpha=@n34.R
   )
   B=LinearInterpolate(
    A=VectorParameter() {group=Projection_SplitWeapon_Uvs, parameterName=SplitW_Offset_Mech, defaultValue=<rgba 0 0 0 1>}.RGB
    B=VectorParameter() {group=Projection_SplitWeapon_Uvs, parameterName=SplitW_Offset_Character, defaultValue=<rgba 0 0 0 1>}.RGB
    Alpha=@n34.A
   )
   Alpha=@n34.B
  )
  n36 = VectorParameter() {group=Projection_SplitWeapon_Uvs, parameterName=SplitW_Offset_Backpack, defaultValue=<rgba 0 0 0 1>}
  n37 = MaterialFunctionCall(Float2=NamedRerouteUsage()) {materialFunction=&BreakOutFloat2Components}
  NamedRerouteDeclaration(None=LinearInterpolate(
    A=LinearInterpolate(A=@n35, B=@n36.RGB, Alpha=@n37.R)
    B=VectorParameter() {group=Projection_SplitWeapon_Uvs, parameterName=SplitW_Offset_Pickaxe, defaultValue=<rgba 0 0 0 1>}.RGB
    Alpha=@n37.G
   )) {name=UVs_Offset_BC, nodeColor=<rgba 0 0.625 1 1>}
  NamedRerouteDeclaration(None=@n11.Normal) {name=Root_Normal, nodeColor=<rgba 0.4062 0 1 1>}
  n38 = VectorParameter() {group=Projection_SplitWeapon_Uvs, parameterName=SplitW_Rotation_Weapon, defaultValue=<rgba 0 0 -0.25 1>}
  n39 = VectorParameter() {group=Projection_SplitWeapon_Uvs, parameterName=SplitW_Rotation_Vehicle, defaultValue=<rgba 0 0 0 1>}
  n40 = VectorParameter() {group=Projection_SplitWeapon_Uvs, parameterName=SplitW_Rotation_Vehicle_Side, defaultValue=<rgba 0 0 0 1>}
  n41 = MaterialFunctionCall(Float4=NamedRerouteUsage()) {materialFunction=&BreakOutFloat4Components}
  n42 = VectorParameter() {group=Projection_SplitWeapon_Uvs, parameterName=SplitW_Rotation_Mech, defaultValue=<rgba 0.25 0.25 0 1>}
  n43 = VectorParameter() {group=Projection_SplitWeapon_Uvs, parameterName=SplitW_Rotation_Character, defaultValue=<rgba 0.25 0.25 0 1>}
  n44 = LinearInterpolate(
   A=LinearInterpolate(A=@n38.RGB, B=LinearInterpolate(A=@n39.RGB, B=@n40.RGB, Alpha=@n41.G), Alpha=@n41.R)
   B=LinearInterpolate(A=@n42.RGB, B=@n43.RGB, Alpha=@n41.A)
   Alpha=@n41.B
  )
  n45 = VectorParameter() {group=Projection_SplitWeapon_Uvs, parameterName=SplitW_Rotation_Backpack, defaultValue=<rgba 0 0 0 1>}
  n46 = MaterialFunctionCall(Float2=NamedRerouteUsage()) {materialFunction=&BreakOutFloat2Components}
  NamedRerouteDeclaration(None=LinearInterpolate(
    A=LinearInterpolate(A=@n44, B=@n45.RGB, Alpha=@n46.R)
    B=VectorParameter() {group=Projection_SplitWeapon_Uvs, parameterName=SplitW_Rotation_Pickaxe, defaultValue=<rgba 0 0 0 1>}.RGB
    Alpha=@n46.G
   )) {name=UVs_TopSide_Rotation, nodeColor=<rgba 0 1 0.5078 1>}
  n47 = ScalarParameter() {defaultValue=1, group=Projection_SplitWeapon_Uvs, parameterName=SplitW_Roughness_Scale_Vehicle, sortPriority=150}
  n48 = ScalarParameter() {defaultValue=1, group=Projection_SplitWeapon_Uvs, parameterName=SplitW_Roughness_Scale_Vehicle_Side, sortPriority=150}
  n49 = MaterialFunctionCall(Float4=NamedRerouteUsage()) {materialFunction=&BreakOutFloat4Components}
  n50 = LinearInterpolate(
   A=LinearInterpolate(
    A=ScalarParameter() {defaultValue=1, group=Projection_SplitWeapon_Uvs, parameterName=SplitW_Roughness_Scale_Weapon, sortPriority=150}
    B=LinearInterpolate(A=@n47, B=@n48, Alpha=@n49.G)
    Alpha=@n49.R
   )
   B=LinearInterpolate(
    A=ScalarParameter() {defaultValue=1, group=Projection_SplitWeapon_Uvs, parameterName=SplitW_Roughness_Scale_Mech, sortPriority=150}
    B=ScalarParameter() {defaultValue=1, group=Projection_SplitWeapon_Uvs, parameterName=SplitW_Roughness_Scale_Character, sortPriority=150}
    Alpha=@n49.A
   )
   Alpha=@n49.B
  )
  n51 = ScalarParameter() {defaultValue=1, group=Projection_SplitWeapon_Uvs, parameterName=SplitW_Roughness_Scale_Backpack, sortPriority=150}
  n52 = MaterialFunctionCall(Float2=NamedRerouteUsage()) {materialFunction=&BreakOutFloat2Components}
  NamedRerouteDeclaration(None=LinearInterpolate(
    A=LinearInterpolate(A=@n50, B=@n51, Alpha=@n52.R)
    B=ScalarParameter() {defaultValue=1, group=Projection_SplitWeapon_Uvs, parameterName=SplitW_Roughness_Scale_Pickaxe, sortPriority=150}
    Alpha=@n52.G
   )) {name=UVs_Scale_Msk, nodeColor=<rgba 0 0.7188 1 1>}
  n53 = FunctionInput(Preview=@n1.G) {bUsePreviewValueAsDefault=true, inputName=Mask Crevise, inputType=FunctionInput_Scalar, sortPriority=24, previewValue=<1 0 0 1>}
  NamedRerouteDeclaration(None=@n53) {name=Msk_Curvature, nodeColor=<rgba 0 1 0.7422 1>}
  NamedRerouteDeclaration(None=FunctionInput(Preview=Constant4Vector().RGBA) {bUsePreviewValueAsDefault=true, inputName=IsState, inputType=FunctionInput_Vector4, sortPriority=4}) {name=IsState, nodeColor=<rgba 0.1497 1 0.1633 1>}
  NamedRerouteDeclaration(None=FunctionInput(Preview=Constant2Vector()) {bUsePreviewValueAsDefault=true, inputName=IsState_B, inputType=FunctionInput_Vector2, sortPriority=4}) {name=IsState_B, nodeColor=<rgba 0.1497 1 0.1633 1>}
  NamedRerouteUsage()
 ]}
