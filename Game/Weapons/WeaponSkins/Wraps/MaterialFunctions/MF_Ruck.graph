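# Ruckus wrap material function for weapon, vehicle and mech skins, adding a glowing core with animated swirls inside the customization mask.
# Virtual-plane UVs, offset per item type and warped by a panning texture, drive distance-based core glows and four counter-rotating mask layers.
# A sphere mask blends core and swirls, then Fresnel rim light is added to emissive.
# The Wrap_Ruck static switch passes input through when off; the secondary-mask blend is left unconnected.
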MaterialFunction /Game/Weapons/WeaponSkins/Wraps/MaterialFunctions/MF_Ruck
graph {nodes [
  n1 = PreSkinnedLocalBounds()
  n2 = Add(A=@n1.Min, B=FunctionInput() {inputName=Object Bounds Min Adjustment, sortPriority=51}) {constB=0}
  n3 = FunctionInput() {inputName=Object Bounds Max Adjustment, sortPriority=52}
  n4 = Multiply(A=Subtract(A=Add(A=@n1.Max, B=@n3) {constB=0}, B=@n2)) {constB=0.5}
  n5 = FunctionInput(Preview=Constant()) {bUsePreviewValueAsDefault=true, inputName=isMech, inputType=FunctionInput_Scalar}
  n6 = LinearInterpolate(A=TransformPosition(None=Add(A=@n2, B=@n4)) {transformType=TRANSFORMPOSSOURCE_World}, B=ObjectPositionWS(), Alpha=@n5)
  n7 = Normalize(VectorInput=Subtract(A=@n6, B=CameraPositionWS()))
  n8 = CameraVectorWS()
  n9 = Ceil(None=Saturate(None=DotProduct(A=@n7, B=@n8)))
  n10 = VectorParameter() {group=Wrap_Ruckus, parameterName=Ruck_Core01_EmColor_01, defaultValue=<rgba 0.295 1 0.3439 1>}
  n11 = VectorParameter() {group=Wrap_Ruckus, parameterName=Ruck_Core01_EmColor_02, defaultValue=<rgba 0 0 0 1>}
  n12 = ScalarParameter() {defaultValue=0.3, group=Wrap_Ruckus, parameterName=Ruck_Core01_DistanceFalloff-----}
  n13 = VectorParameter() {group=Wrap_Ruckus_UV, parameterName=Ruck_Offset_Vehicle}
  n14 = VectorParameter() {group=Wrap_Ruckus_UV, parameterName=Ruck_Offset_VehicleSide}
  n15 = Constant()
  n16 = Constant()
  n17 = MaterialFunctionCall(Color=MaterialFunctionCall(
    R=FunctionInput(Preview=@n15) {inputName=isVehicle, inputType=FunctionInput_Scalar, sortPriority=5}
    G=FunctionInput(Preview=@n16) {bUsePreviewValueAsDefault=true, inputName=Side Project Vehicle, inputType=FunctionInput_Scalar}
    B=@n5
   ) {materialFunction=&AppendMany}.RGB) {materialFunction=&SplitComponents}
  n18 = LinearInterpolate(
   A=LinearInterpolate(
    A=VectorParameter() {group=Wrap_Ruckus_UV, parameterName=Ruck_Offset_Weapon}.RGB
    B=LinearInterpolate(A=@n13.RGB, B=@n14.RGB, Alpha=@n17.G)
    Alpha=@n17.R
   )
   B=VectorParameter() {group=Wrap_Ruckus_UV, parameterName=Ruck_Offset_Mech}.RGB
   Alpha=@n17.B
  )
  n19 = NamedRerouteUsage()
  n20 = Time()
  n21 = ComponentMask(None=VectorParameter() {group=Wrap_Ruckus, parameterName=Ruck_Warp_Speed-----, defaultValue=<rgba 0.1 0.1 0 1>}.RGB) {g=true, r=true}
  n22 = TextureObjectParameter() {group=Wrap_Ruckus, parameterName=Ruckus_Warp_Texture, sortPriority=-100, texture=&T_BlacksRGB}
  n23 = TextureSample(UVs=Add(A=TextureCoordinate(), B=Multiply(A=@n20, B=@n21)), Tex=@n22) {samplerSource=SSM_Clamp_WorldGroupSettings, samplerType=SAMPLERTYPE_Masks}
  n24 = ScalarParameter() {defaultValue=0.1, group=Wrap_Ruckus, parameterName=Ruck_WarpUVS_01_Amount Weapon}
  n25 = ScalarParameter() {defaultValue=0.1, group=Wrap_Ruckus, parameterName=Ruck_WarpUVS_01_Amount Vehicle}
  n26 = Add(
   A=Add(A=ComponentMask(None=@n18) {g=true, r=true}, B=@n19)
   B=Multiply(
    A=ChannelMaskParameter(None=@n23.RGBA) {group=Wrap_Ruckus, parameterName=Ruck_WarpUVS_01_Channel}.R
    B=LinearInterpolate(A=@n24, B=@n25, Alpha=@n17.R)
   )
  )
  n27 = Fresnel(ExponentIn=ScalarParameter() {defaultValue=0.4967, group=Wrap_Ruckus, parameterName=Ruck_Core00_DistortionExp-----}) {baseReflectFraction=0}
  n28 = ScalarParameter() {defaultValue=0.27, group=Wrap_Ruckus, parameterName=Ruck_Core00_DistortionNormal}
  n29 = Multiply(A=Distance(A=@n26, B=Constant2Vector() {g=0.5, r=0.5}), B=OneMinus(None=Multiply(A=@n27, B=@n28)))
  n30 = Clamp(
   None=Power(Base=Divide(A=@n29, B=@n12), Exp=ScalarParameter() {defaultValue=80, group=Wrap_Ruckus, parameterName=Ruck_Core01_Fresnel_Exp})
   Min=ScalarParameter() {defaultValue=0.9, group=Wrap_Ruckus, parameterName=Ruck_Core01_MaxBrightness}
  )
  n31 = VectorParameter() {group=Wrap_Ruckus, parameterName=Ruck_Core02_CoreGlow------, defaultValue=<rgba 0.2811 0.6562 0.2738 0>}
  n32 = Multiply(A=@n31.RGB, B=ScalarParameter() {defaultValue=0.45, group=Wrap_Ruckus, parameterName=Ruck_Core02_Glow})
  n33 = Divide(A=@n29, B=ScalarParameter() {defaultValue=1, group=Wrap_Ruckus, parameterName=Ruck_Core02_DistanceFalloff})
  n34 = Clamp(None=Power(Base=OneMinus(None=@n33), Exp=ScalarParameter() {defaultValue=18, group=Wrap_Ruckus, parameterName=Ruck_Core02_Fresnel_Exp}))
  n35 = ScalarParameter() {defaultValue=1.3, group=Wrap_Ruckus, parameterName=Ruck_Core03_DistanceFalloff}
  n36 = Power(Base=OneMinus(None=Divide(A=@n29, B=@n35)), Exp=ScalarParameter() {defaultValue=3, group=Wrap_Ruckus, parameterName=Ruck_Core03_Fresnel_Exp})
  n37 = Add(
   A=Add(A=LinearInterpolate(A=@n10.RGB, B=@n11.RGB, Alpha=@n30), B=Multiply(A=@n32, B=@n34))
   B=Multiply(
    A=VectorParameter() {group=Wrap_Ruckus, parameterName=Ruck_Core03_CoreGlow-----, defaultValue=<rgba 0.1144 0.7529 0.0823 1>}.RGB
    B=Clamp(None=@n36)
   )
  )
  n38 = VectorParameter() {group=Wrap_Ruckus, parameterName=Ruck_Twirl_Channel_01_Basecolor, defaultValue=<rgba 0.1687 0.5938 0.1345 1>}
  n39 = VectorParameter() {group=Wrap_Ruckus, parameterName=Ruck_Twirl_Channel_04_Color, defaultValue=<rgba 0.0137 0.2664 0.0578 1>}
  n40 = MaterialFunctionCall(
   UVs=@n26
   Texture Scale=LinearInterpolate(
    A=ScalarParameter() {defaultValue=2, group=Wrap_Ruckus, parameterName=Ruck_Twirl_Scale}
    B=ScalarParameter() {defaultValue=2, group=Wrap_Ruckus, parameterName=Ruck_Twirl_Scale_Vehicle}
    Alpha=@n17.R
   )
  ) {materialFunction=&ScaleUVsByCenter}
  n41 = Constant2Vector() {g=0.5, r=0.5}
  n42 = Time()
  n43 = Multiply(A=@n42, B=ScalarParameter() {defaultValue=-0.3, group=Wrap_Ruckus, parameterName=Ruck_Twirl_Channel_04_RotSpeed})
  n44 = TextureObjectParameter() {group=Wrap_Ruckus, parameterName=Ruckus_Mask__Text, sortPriority=-100, texture=&T_BlacksRGB}
  n45 = ChannelMaskParameter(None=TextureSample(
    UVs=MaterialFunctionCall(UVs=@n40.UVs, Rotation Center=@n41, "Rotation Angle (0-1)"=@n43) {materialFunction=&CustomRotator}.Rotated Values
    Tex=@n44
   ) {samplerSource=SSM_Clamp_WorldGroupSettings, samplerType=SAMPLERTYPE_Masks}.RGBA) {group=Wrap_Ruckus, maskChannel=Alpha, parameterName=Ruck_Twirl_Channel_04, defaultValue=<rgba 0 0 0 1>}
  n46 = LinearInterpolate(A=@n38.RGB, B=@n39.RGB, Alpha=@n45.R)
  n47 = VectorParameter() {group=Wrap_Ruckus, parameterName=Ruck_Twirl_Channel_03_Color, defaultValue=<rgba 0.107 0.4125 0.0666 1>}
  n48 = Multiply(A=ScalarParameter() {defaultValue=-0.1, group=Wrap_Ruckus, parameterName=Ruck_Twirl_Channel_03_RotSpeed}, B=@n42)
  n49 = ChannelMaskParameter(None=TextureSample(
    UVs=MaterialFunctionCall(UVs=@n40.UVs, Rotation Center=@n41, "Rotation Angle (0-1)"=@n48) {materialFunction=&CustomRotator}.Rotated Values
    Tex=@n44
   ) {samplerSource=SSM_Clamp_WorldGroupSettings, samplerType=SAMPLERTYPE_Masks}.RGBA) {group=Wrap_Ruckus, maskChannel=Blue, parameterName=Ruck_Twirl_Channel_03, defaultValue=<rgba 0 0 1 0>}
  n50 = VectorParameter() {group=Wrap_Ruckus, parameterName=Ruck_Twirl_Channel_02_Color, defaultValue=<rgba 0.3916 0.5972 0.0908 1>}
  n51 = Constant2Vector() {g=0.5, r=0.5}
  n52 = Time()
  n53 = Multiply(A=@n52, B=ScalarParameter() {defaultValue=-0.3, group=Wrap_Ruckus, parameterName=Ruck_Twirl_Channel_02_RotSpeed})
  n54 = ChannelMaskParameter(None=TextureSample(
    UVs=MaterialFunctionCall(UVs=@n40.UVs, Rotation Center=@n51, "Rotation Angle (0-1)"=@n53) {materialFunction=&CustomRotator}.Rotated Values
    Tex=@n44
   ) {samplerSource=SSM_Clamp_WorldGroupSettings, samplerType=SAMPLERTYPE_Masks}.RGBA) {group=Wrap_Ruckus, maskChannel=Green, parameterName=Ruck_Twirl_Channel_02, defaultValue=<rgba 0 1 0 0>}
  n55 = MaterialFunctionCall(
   UVs=@n40.UVs
   Rotation Center=@n51
   "Rotation Angle (0-1)"=Multiply(A=ScalarParameter() {defaultValue=0.1, group=Wrap_Ruckus, parameterName=Ruck_Twirl_Channel_01_RotSpeed}, B=@n52)
  ) {materialFunction=&CustomRotator}
  n56 = LinearInterpolate(
   A=LinearInterpolate(A=LinearInterpolate(A=@n46, B=@n47.RGB, Alpha=@n49.R), B=@n50.RGB, Alpha=@n54.R)
   B=VectorParameter() {group=Wrap_Ruckus, parameterName=Ruck_Twirl_Channel_01_Color, defaultValue=<rgba 1 1 1 1>}.RGB
   Alpha=ChannelMaskParameter(None=TextureSample(UVs=@n55.Rotated Values, Tex=@n44) {samplerSource=SSM_Clamp_WorldGroupSettings, samplerType=SAMPLERTYPE_Masks}.RGBA) {group=Wrap_Ruckus, parameterName=Ruck_Twirl_Channel_01-----}.R
  )
  n57 = ScalarParameter() {defaultValue=-0.1, group=Wrap_Ruckus, parameterName=Ruck_BrightnessSaturation}
  Add(
   A=Multiply(A=OneMinus(None=@n9), B=@n37)
   B=Multiply(A=Desaturation(None=@n56, Fraction=@n57), B=ScalarParameter() {defaultValue=0.1, group=Wrap_Ruckus, parameterName=Ruck_BrightnessSwirl})
  )
  n58 = FunctionInput() {bUsePreviewValueAsDefault=true, inputName=isCharacter, inputType=FunctionInput_Scalar}
  Add(A=@n58) {constB=0}
  n59 = FunctionInput() {inputName=Material, inputType=FunctionInput_MaterialAttributes, sortPriority=-200}
  n60 = GetMaterialAttributes(None=@n59)
  n61 = GetMaterialAttributes(None=@n59) {attributeGetTypes [#1]}
  n62 = Multiply(
   A=VectorParameter() {group=Wrap_Ruckus, parameterName=Ruck_Fresnel_Color, sortPriority=1, defaultValue=<rgba 1 1 1 1>}.RGB
   B=Fresnel(
    ExponentIn=ScalarParameter() {defaultValue=8, group=Wrap_Ruckus, parameterName=Ruck_Fresnel -----, sortPriority=1}
    Normal=Transform(None=@n61.Normal)
   ) {baseReflectFraction=0}
  )
  n63 = ScalarParameter() {defaultValue=1, group=Wrap_Ruckus, parameterName=Ruck_Fresnel Intensity, sortPriority=1}
  n64 = SphereMask(
   A=@n19
   B=Constant() {r=0.5}
   Radius=ScalarParameter() {defaultValue=1.2, group=Wrap_Ruckus, parameterName=Ruck_Core_SphereMask_Radius}
   Hardness=ScalarParameter() {defaultValue=75, group=Wrap_Ruckus, parameterName=Ruck_Core_SphereMask_Hardness-----}
  ) {attenuationRadius=1.2}
  n65 = Multiply(
   A=Add(A=Multiply(A=@n62, B=@n63), B=LinearInterpolate(A=@n37, B=@n56, Alpha=@n64)) {constB=0}
   B=ScalarParameter() {defaultValue=0.5, group=Wrap_Ruckus, parameterName=Ruck_Emissive_Amount}
  )
  n66 = SetMaterialAttributes!(
   MaterialAttributes=@n60.MaterialAttributes
   BaseColor=VectorParameter() {group=Wrap_Ruckus, parameterName=Ruck_BaseColor-----, sortPriority=0}.RGB
   Roughness=ScalarParameter() {defaultValue=0.1, group=Wrap_Ruckus, parameterName=Ruck_Roughness, sortPriority=0}
   Specular=ScalarParameter() {defaultValue=0.5, group=Wrap_Ruckus, parameterName=Ruck_Specular, sortPriority=0}
   EmissiveColor=@n65
  ) {attributeSetTypes [
    #2
    #3
    #4
    #5
   ]}
  n67 = FunctionInput() {inputName="Customization Mask (_CM B)", inputType=FunctionInput_Scalar, sortPriority=24, previewValue=<1 0 0 1>}
  n68 = BlendMaterialAttributes(A=@n60.MaterialAttributes, B=@n66, Alpha=@n67)
  n69 = VectorParameter() {group=Wrap_Ruckus, parameterName="Ruck_Secondary_BaseColor A=Roughness-----", sortPriority=0}
  n70 = SetMaterialAttributes!(
   MaterialAttributes=GetMaterialAttributes(None=@n68).MaterialAttributes
   BaseColor=@n69.RGB
   Metallic=ScalarParameter() {group=Wrap_Ruckus, parameterName=Ruck_Secondary_Metallic, sortPriority=0}
   Specular=ScalarParameter() {defaultValue=0.5, group=Wrap_Ruckus, parameterName=Ruck_Secondary_Specular, sortPriority=0}
   Roughness=@n69.RGB
  ) {attributeSetTypes [
    #2
    #6
    #4
    #3
   ]}
  n71 = FunctionInput() {inputName="Customization Mask (_Secondary)", inputType=FunctionInput_Scalar, sortPriority=24, previewValue=<1 0 0 1>}
  BlendMaterialAttributes(A=@n68, B=@n70, Alpha=@n71)
  FunctionOutput(None=StaticSwitchParameter(True=@n68, False=@n59) {group=Customization_Wrap, parameterName=Wrap_Ruck})
  n72 = CrossProduct(A=@n7, B=VectorParameter() {group=Wrap_Ruckus_UV, parameterName=Ruck_Axis, defaultValue=<rgba 0 0 1 1>}.RGB)
  n73 = MaterialFunctionCall() {materialFunction=/Engine/Functions/Engine_MaterialFunctions02/WorldPositionOffset/ObjectScale}
  n74 = MaterialFunctionCall(Float3=PreSkinnedLocalBounds().Half Extents) {materialFunction=&BreakOutFloat3Components}
  n75 = Max(A=Max(A=@n74.R, B=@n74.R), B=@n74.B)
  LinearInterpolate(
   A=MaterialFunctionCall!(Plane Normal Axis=@n7, Plane X Axis=Normalize(VectorInput=@n72), Plane Center=@n6, UV Scale=Multiply(A=@n73.Scale X, B=@n75)) {materialFunction=/Game/Athena/Environments/Nexus/Materials/VirtualPlaneCoordinatesAbs}.UVs
   B=@n19
   Alpha=@n58
  )
  Multiply(
   A=@n65
   B=MaterialFunctionCall(Value_Morning=Constant() {r=0.5}, Value_Day=Constant() {r=1}, Value_Evening=Constant() {r=0.5}, Value_Night=Constant() {r=0.5}) {materialFunction=&Scalar_Time_of_Day}.Result
  )
  NamedRerouteDeclaration(None=@n58) {name=IsCharacter, nodeColor=<rgba 0 0.6484 1 1>}
  n76 = FunctionInput() {bUsePreviewValueAsDefault=true, inputName=IsStatsB, inputType=FunctionInput_Vector2, sortPriority=-19}
  NamedRerouteDeclaration(None=ComponentMask(None=@n76) {r=true}) {name=IsBackpack, nodeColor=<rgba 0 0.9253 1 1>}
  NamedRerouteDeclaration(None=ComponentMask(None=@n76) {g=true}) {name=IsPickaxe, nodeColor=<rgba 0 0.9253 1 1>}
  n77 = FunctionInput() {bUsePreviewValueAsDefault=true, inputName=Object-SafeZone-Virtual-Pivot, inputType=FunctionInput_Vector4, sortPriority=105}
  n78 = FunctionInput() {bUsePreviewValueAsDefault=true, inputName=Object-SafeZone-Virtual-CH-Switch, inputType=FunctionInput_Scalar, sortPriority=105}
  NamedRerouteDeclaration(None=MaterialFunctionCall!(
    Camera _pos
    Axis
    Scale
    Scale_Multplyer
    Object-SafeZone-Virtual-Pivot=@n77
    Object-SafeZone-Virtual-CH-Switch=@n78
    SinglePlane Normal Axis
    SinglePlane X Axis
    PlaneCenter_distortion
   ) {materialFunction=&/MF_SafeZone_VirtualPlane}.Uvs) {name=SafeZone_VP_UVs, nodeColor=<rgba 0.6172 1 0 1>}
  NamedRerouteUsage()
 ]}
