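# Material function for the ShinyStar cosmetic wrap on weapons, vehicles, characters, backpacks and pickaxes, enabled by a static switch.
# Side-projected UVs, adjusted per item type via IsState inputs, drive a normal-mapped star decal and radial scrolling stars.
# Three warped, contrast-boosted polar streams tint a radial background gradient.
# Base and accent materials blend by input masks, with curvature brightening edges and a timed reactive fade available for glow.
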MaterialFunction /Game/Weapons/WeaponSkins/Wraps/MaterialFunctions/MF_ShinySt
graph {nodes [
  n1 = Add(A=NamedRerouteUsage()) {constB=-0.5}
  n2 = Multiply(
   A=Add(A=MaterialFunctionCall(Vector or UVs=@n1) {materialFunction=&VectorToRadialValue}.Radial Coordinates) {constB=0.35}
   B=ComponentMask(None=VectorParameter() {group=Wrap_ShinyStar_Streams, parameterName=ShinyS_StreamB_TileRate, defaultValue=<rgba 2 0.1 0 1>}.RGB) {g=true, r=true}
  )
  n3 = ComponentMask(None=Multiply(A=Time(), B=VectorParameter() {group=Wrap_ShinyStar_Streams, parameterName=ShinyS_StreamB_Speed, defaultValue=<rgba -0.025 -0.25 0 1>}.RGB)) {g=true, r=true}
  n4 = Add(A=@n2, B=@n3)
  n5 = NamedRerouteUsage()
  n6 = NamedRerouteUsage()
  n7 = NamedRerouteUsage()
  n8 = Add(A=Add(A=@n2, B=Constant2Vector() {g=0.5, r=0.5}), B=@n3)
  n9 = Add(
   A=TextureSample!(UVs=Add(A=@n4, B=@n5), Tex=@n6, Level=@n7, Apply View MipBias) {mipValueMode=TMVM_MipLevel, samplerType=SAMPLERTYPE_LinearGrayscale}.R
   B=TextureSample(UVs=Add(A=@n8, B=@n5), Tex=@n6, Level=@n7) {mipValueMode=TMVM_MipLevel, samplerType=SAMPLERTYPE_Masks}.R
  )
  Add(
   A=@n9
   B=SmoothStep(Value=MaterialFunctionCall(Vector 2=@n1) {materialFunction=/Engine/Functions/Engine_MaterialFunctions02/Utility/VectorLength}.V2 Length) {constMax=0, constMin=0.08}
  )
  n10 = Add(A=NamedRerouteUsage()) {constB=-0.5}
  n11 = Multiply(
   A=Add(A=MaterialFunctionCall(Vector or UVs=@n10) {materialFunction=&VectorToRadialValue}.Radial Coordinates) {constB=0.1}
   B=ComponentMask(None=VectorParameter() {group=Wrap_ShinyStar_Streams, parameterName=ShinyS_StreamC_TileRate, defaultValue=<rgba 1 0.1 0 1>}.RGB) {g=true, r=true}
  )
  n12 = ComponentMask(None=Multiply(A=Time(), B=VectorParameter() {group=Wrap_ShinyStar_Streams, parameterName=ShinyS_StreamC_Speed, defaultValue=<rgba 0.01 -0.25 0 1>}.RGB)) {g=true, r=true}
  n13 = Add(A=@n11, B=@n12)
  n14 = NamedRerouteUsage()
  n15 = NamedRerouteUsage()
  n16 = NamedRerouteUsage()
  n17 = Add(A=Add(A=@n11, B=Constant2Vector() {g=0.5, r=0.5}), B=@n12)
  n18 = Add(
   A=TextureSample(UVs=Add(A=@n13, B=@n14), Tex=@n15, Level=@n16) {mipValueMode=TMVM_MipLevel, samplerType=SAMPLERTYPE_LinearGrayscale}.R
   B=TextureSample(UVs=Add(A=@n17, B=@n14), Tex=@n15, Level=@n16) {mipValueMode=TMVM_MipLevel, samplerType=SAMPLERTYPE_Masks}.R
  )
  Add(
   A=@n18
   B=SmoothStep(Value=MaterialFunctionCall(Vector 2=@n10) {materialFunction=&VectorLength}.V2 Length) {constMax=0, constMin=0.08}
  )
  n19 = NamedRerouteUsage()
  n20 = StaticBool() {value=true}
  n21 = MaterialFunctionCall!(
   Normal_Map=NamedRerouteUsage()
   Offset Rt
   Pivot
   Caculated_Bounds=NamedRerouteUsage()
   Pre_Skinned_Normals=NamedRerouteUsage()
   is_State=NamedRerouteUsage()
   UV_Rotation=NamedRerouteUsage()
   Uvs_Offset=ComponentMask(None=@n19) {b=true, g=true, r=true}
   Uvs_TileRate=ComponentMask(None=@n19) {a=true}
   Uvs_Offset_Pivot
   For_Safe_Zones=@n20
   Rotation_Comp
   Clamp Y=@n20
   Clamp X=@n20
   Clamp Padding_Comp
   No_Rotation
  ) {materialFunction=&/MF_Projection_Side_Normal_V2}
  n22 = TextureSample!(UVs=@n21.UVs, Tex=NamedRerouteUsage(), Apply View MipBias) {samplerType=SAMPLERTYPE_Normal}
  n23 = Multiply(A=@n22.A, B=NamedRerouteUsage())
  Add(A=@n23, B=NamedRerouteUsage())
  n24 = Add(A=Add(A=NamedRerouteUsage(), B=NamedRerouteUsage()), B=NamedRerouteUsage())
  Add(A=@n24, B=NamedRerouteUsage())
  n25 = TextureSample() {samplerType=SAMPLERTYPE_Masks, texture=/Game/Weapons/FORT_Rifles/Textures/T_Scar_CM}
  n26 = FunctionInput(Preview=@n25.G) {bUsePreviewValueAsDefault=true, inputName=Mask Crevise, inputType=FunctionInput_Scalar, sortPriority=24, previewValue=<1 0 0 1>}
  n27 = NamedRerouteDeclaration(None=@n26) {name=Msk_Curvature, nodeColor=<rgba 0 1 0.7422 1>}
  Add(A=OneMinus(None=@n27)) {constB=-0.5}
  n28 = Add(A=NamedRerouteUsage()) {constB=-0.5}
  n29 = Multiply(
   A=MaterialFunctionCall(Vector or UVs=@n28) {materialFunction=&VectorToRadialValue}.Radial Coordinates
   B=ComponentMask(None=VectorParameter() {group=Wrap_ShinyStar_Streams, parameterName=ShinyS_StreamA_TileRate, defaultValue=<rgba 1.5 0.1 0 1>}.RGB) {g=true, r=true}
  )
  n30 = ComponentMask(None=Multiply(A=Time(), B=VectorParameter() {group=Wrap_ShinyStar_Streams, parameterName=ShinyS_StreamA_Speed, defaultValue=<rgba 0.01 -0.25 0 1>}.RGB)) {g=true, r=true}
  n31 = Add(A=@n29, B=@n30)
  n32 = NamedRerouteUsage()
  n33 = NamedRerouteUsage()
  n34 = NamedRerouteUsage()
  n35 = Add(A=Add(A=@n29, B=Constant2Vector() {g=0.5, r=0.5}), B=@n30)
  n36 = Add(
   A=TextureSample!(UVs=Add(A=@n31, B=@n32), Tex=@n33, Level=@n34, Apply View MipBias) {mipValueMode=TMVM_MipLevel, samplerType=SAMPLERTYPE_LinearGrayscale}.G
   B=TextureSample(UVs=Add(A=@n35, B=@n32), Tex=@n33, Level=@n34) {mipValueMode=TMVM_MipLevel, samplerType=SAMPLERTYPE_Masks}.G
  )
  Add(
   A=@n36
   B=SmoothStep(Value=MaterialFunctionCall(Vector 2=@n28) {materialFunction=&VectorLength}.V2 Length) {constMax=0, constMin=0.08}
  )
  Constant() {r=1}
  n37 = NamedRerouteUsage()
  n38 = NamedRerouteUsage()
  n39 = FunctionInput(Preview=@n25.R) {bUsePreviewValueAsDefault=true, inputName=Mask Primary, inputType=FunctionInput_Scalar, sortPriority=24, previewValue=<1 0 0 1>}
  n40 = FunctionInput(Preview=@n25.B) {bUsePreviewValueAsDefault=true, inputName=Mask Secondary, inputType=FunctionInput_Scalar, sortPriority=24, previewValue=<1 0 0 1>}
  n41 = Saturate(None=Add(A=@n39, B=@n40))
  n42 = GetMaterialAttributes(None=BlendMaterialAttributes(
    A=BlendMaterialAttributes(A=@n37, B=@n38, Alpha=@n41) {vertexAttributeBlendType=UseA}
    B=NamedRerouteUsage()
    Alpha=NamedRerouteDeclaration(None=@n40) {name=Msk_Accent, nodeColor=<rgba 0 1 0.7656 1>}
   ) {vertexAttributeBlendType=UseA}) {attributeGetTypes [#1, #2]}
  n43 = ScalarParameter() {defaultValue=2, group=Wrap_ShinyStar, parameterName=ShintS_Edge_Brightness}
  n44 = Add(A=@n27) {constB=-0.5}
  n45 = LinearInterpolate(A=@n42.BaseColor, B=Multiply(A=@n42.BaseColor, B=@n43), Alpha=@n44)
  n46 = LinearInterpolate(A=@n42.Roughness, B=Multiply(A=@n42.Roughness, B=@n43), Alpha=Saturate(None=@n44))
  FunctionOutput(None=StaticSwitchParameter(
    True=SetMaterialAttributes!(MaterialAttributes=@n42.MaterialAttributes, BaseColor=@n45, Roughness=@n46) {attributeSetTypes [#1, #2]}
    False=FunctionInput() {inputName=Material, inputType=FunctionInput_MaterialAttributes}
   ) {group=Customization_Wrap, parameterName=Wrap_ShinyStar})
  LinearInterpolate(B=@n21.Raw_Normal, Alpha=@n23)
  n47 = Multiply(A=Time(), B=VectorParameter() {group=Wrap_ShinyStar, parameterName=ShinyS_Stars_Speed, defaultValue=<rgba 0.05 -0.25 0 1>}.RGB)
  n48 = MaterialFunctionCall(
   Caculated_Bounds=NamedRerouteUsage()
   Pre_Skinned_Normals=NamedRerouteUsage()
   is_State=NamedRerouteUsage()
   Uvs_Offset=ComponentMask(None=NamedRerouteUsage()) {b=true, g=true, r=true}
   Uvs_TileRate=NamedRerouteUsage()
   For_Safe_Zones=StaticBool() {value=true}
  ) {materialFunction=&MF_Projection_Side_Normal_V2}
  n49 = Add(A=NamedRerouteDeclaration(None=@n48.UVs) {name=Uvs, nodeColor=<rgba 1 0.6563 0 1>}) {constB=-0.5}
  n50 = MaterialFunctionCall(Vector or UVs=@n49) {materialFunction=&VectorToRadialValue}
  n51 = VectorParameter() {group=Wrap_ShinyStar, parameterName=ShinyS_Stars_TileRate, defaultValue=<rgba 1 2 1 0>}
  n52 = Power(Base=MaterialFunctionCall(Vector 2=Multiply(A=@n49) {constB=0.5}) {materialFunction=&VectorLength}.V2 Length) {constExponent=0.75}
  n53 = Multiply(A=AppendVector(A=@n51.R, B=LinearInterpolate(Alpha=@n52) {constA=1, constB=0}), B=@n51.G) {constB=1.5}
  n54 = TextureSample(
   UVs=Add(A=ComponentMask(None=@n47) {g=true, r=true}, B=Multiply(A=@n50.Radial Coordinates, B=@n53))
   Tex=NamedRerouteUsage()
   Level=NamedRerouteUsage()
  ) {mipValueMode=TMVM_MipLevel}
  n55 = Add(A=NamedRerouteUsage()) {constB=-0.5}
  n56 = Power(Base=Saturate(None=MaterialFunctionCall(Vector 2=@n55) {materialFunction=&VectorLength}.V2 Length))
  n57 = Multiply(A=Multiply(A=@n54.A, B=NamedRerouteUsage()), B=SmoothStep(Value=@n56) {constMax=0.3, constMin=0.1}) {constA=1}
  n58 = VectorParameter() {group=Wrap_ShinyStar_Streams, parameterName=ShinyS_StreamC_Color, defaultValue=<rgba 0 0.0382 0.2656 1>}
  LinearInterpolate(
   A=LinearInterpolate(B=@n54.RGB, Alpha=@n57)
   B=@n58.RGB
   Alpha=NamedRerouteDeclaration(None=SmoothStep(Value=@n56) {constMax=0.01, constMin=0.2}) {name=Msk_Glow, nodeColor=<rgba 0.1016 0 1 1>}
  )
  MaterialFunctionCall(Float2=NamedRerouteUsage()) {materialFunction=&BreakOutFloat2Components}
  n59 = Subtract(A=@n50.Vector Converted to Angle) {constB=0.5}
  n60 = OneMinus(None=Abs(None=Multiply(A=@n59) {constB=2}))
  Multiply(A=Saturate(None=Divide(A=@n60) {constB=0.01}), B=@n57)
  n61 = VectorParameter() {group=Wrap_ShinyStar, parameterName=ShinyS_BK_Color_B, defaultValue=<rgba 0.05 0.005 0.005 1>}
  n62 = VectorParameter() {group=Wrap_ShinyStar, parameterName=ShinyS_BK_Color_A, defaultValue=<rgba 0.6 0.03 0.03 1>}
  n63 = MaterialFunctionCall(Vector 2=Add(A=NamedRerouteUsage()) {constB=-0.5}) {materialFunction=&VectorLength}
  n64 = SmoothStep(
   Min=ScalarParameter() {group=Wrap_ShinyStar, parameterName=ShinyS_BK_Color_Min}
   Max=ScalarParameter() {defaultValue=1, group=Wrap_ShinyStar, parameterName=ShinyS_BK_Color_Max}
   Value=@n63.V2 Length
  ) {constMax=0.3, constMin=0.1}
  n65 = Saturate(None=Add(A=@n64, B=@n61.A))
  n66 = VectorParameter() {group=Wrap_ShinyStar_Streams, parameterName=ShinyS_StreamB_Color, defaultValue=<rgba 0.2396 0.005 0.0133 1>}
  n67 = LinearInterpolate(
   A=LinearInterpolate(
    A=LinearInterpolate(A=@n61.RGB, B=@n62.RGB, Alpha=@n65)
    B=VectorParameter() {group=Wrap_ShinyStar_Streams, parameterName=ShinyS_StreamA_Color, defaultValue=<rgba 0.4 0.4 0.4 1>}.RGB
    Alpha=NamedRerouteUsage()
   )
   B=@n66.RGB
   Alpha=NamedRerouteUsage()
  )
  Multiply(A=@n67)
  Multiply(A=@n64)
  Multiply(A=MaterialFunctionCall(Value_Morning=Constant() {r=0.5}, Value_Day=Constant() {r=1}, Value_Evening=Constant() {r=0.5}, Value_Night=Constant() {r=0.25}) {materialFunction=&Scalar_Time_of_Day}.Result)
  n68 = VectorParameter() {group=Wrap_ShinyStar_Accent, parameterName=ShinyS_Accent_BaseColor, defaultValue=<rgba 0.02 0.02 0.02 1>}
  n69 = SetMaterialAttributes!(
   MaterialAttributes=NamedRerouteUsage()
   BaseColor=@n68.RGB
   Roughness=ScalarParameter() {defaultValue=0.5, group=Wrap_ShinyStar_Accent, parameterName=ShinyS_Accent_Roughness}
   Specular=ScalarParameter() {defaultValue=0.5, group=Wrap_ShinyStar_Accent, parameterName=ShinyS_Accent_Spec}
   Metallic=ScalarParameter() {group=Wrap_ShinyStar_Accent, parameterName=ShinyS_Accent_Metalic}
   Normal
   EmissiveColor
  ) {attributeSetTypes [
    #1
    #2
    #3
    #4
    #5
    #6
   ]}
  NamedRerouteDeclaration(None=@n69) {name=Mat_Accent, nodeColor=<rgba 0 0.3906 1 1>}
  n70 = TextureSample!(UVs, Tex, Apply View MipBias) {samplerType=SAMPLERTYPE_Normal, texture=/Game/Weapons/FORT_Rifles/T_SCAR_N}
  n71 = Constant3Vector() {constant=<rgba 1 0 0 0>}
  n72 = GetMaterialAttributes(None=FunctionInput(Preview=SetMaterialAttributes!(MaterialAttributes, Normal=@n70.RGB, BaseColor=@n71) {attributeSetTypes [#5, #1]}) {inputName=Material Root, inputType=FunctionInput_MaterialAttributes}) {attributeGetTypes [
    #1
    #2
    #5
    #4
   ]}
  NamedRerouteDeclaration(None=@n72.MaterialAttributes) {name=Root_Material, nodeColor=<rgba 0.7812 1 0 1>}
  NamedRerouteDeclaration(None=TextureObjectParameter() {group=Wrap_ShinyStar, parameterName=ShinyS__Tex_Msk, texture=&T_BlacksRGB}) {name=Tex_Msk, nodeColor=<rgba 0.5938 0 1 1>}
  n73 = TextureObjectParameter() {
   group=Wrap_ShinyStar
   parameterName=ShinyS__Tex_N
   samplerType=SAMPLERTYPE_Normal
   texture=&FlatNormalBlue
  }
  NamedRerouteDeclaration(None=@n73) {name=Tex_N, nodeColor=<rgba 0.5938 0 1 1>}
  n74 = Add(A=NamedRerouteUsage()) {constB=-0.5}
  n75 = MaterialFunctionCall(Vector 2=@n74) {materialFunction=&VectorLength}
  n76 = VectorParameter() {group=Wrap_ShinyStar_Uvs_Warp, parameterName=ShinyS_Warp_TileRate, defaultValue=<rgba 1 2 0 1>}
  n77 = Multiply(A=@n76.RGB, B=@n76.RGB)
  n78 = Time()
  n79 = VectorParameter() {group=Wrap_ShinyStar_Uvs_Warp, parameterName=ShinyS_Warp_Speed, defaultValue=<rgba 0.25 0 0 1>}
  n80 = Add(
   A=Multiply(
    A=MaterialFunctionCall(Vector or UVs=@n74) {materialFunction=&VectorToRadialValue}.Radial Coordinates
    B=ComponentMask(None=@n77) {g=true, r=true}
   )
   B=ComponentMask(None=Multiply(A=@n78, B=@n79.RGB)) {g=true, r=true}
  )
  n81 = NamedRerouteUsage()
  n82 = NamedRerouteUsage()
  n83 = Multiply(
   A=Add(A=TextureSample!(UVs=@n80, Tex=@n81, Level=@n82, Apply View MipBias) {mipValueMode=TMVM_MipLevel, samplerType=SAMPLERTYPE_Masks}.A) {constB=-0.5}
   B=ScalarParameter() {defaultValue=0.05, group=Wrap_ShinyStar_Uvs_Warp, parameterName=ShinyS_Warp_Strentgh}
  )
  NamedRerouteDeclaration(None=Multiply(A=SmoothStep(Value=@n75.V2 Length) {constMax=0.2}, B=@n83)) {name=Warp_Uvs, nodeColor=<rgba 0.3359 1 0 1>}
  n84 = Saturate(None=@n9)
  n85 = ScalarParameter() {defaultValue=5, group=Wrap_ShinyStar_Streams, parameterName=ShinyS_StramB_Contrast}
  NamedRerouteDeclaration(None=Multiply(
    A=NamedRerouteUsage()
    B=MaterialFunctionCall(In=@n84, Contrast=@n85) {materialFunction=&CheapContrast}.Result
   )) {name=StreamsB, nodeColor=<rgba 0.875 1 0 1>}
  n86 = TextureObjectParameter() {
   group=Wrap_ShinyStar
   parameterName=ShinyS__Tex_N1
   samplerType=SAMPLERTYPE_Normal
   texture=&FlatNormalBlue
  }
  NamedRerouteDeclaration(None=@n86) {name=Tex_N1, nodeColor=<rgba 0.5938 0 1 1>}
  n87 = ScalarParameter() {group=Wrap_ShinyStar_Uvs, parameterName="ShinyS_Rotation_Weapon\\Pickaxe", sortPriority=150}
  n88 = MaterialFunctionCall(Float4=NamedRerouteUsage()) {materialFunction=&BreakOutFloat4Components}
  n89 = LinearInterpolate(
   A=ScalarParameter() {group=Wrap_ShinyStar_Uvs, parameterName=ShinyS_Rotation_Vehicle, sortPriority=150}
   B=ScalarParameter() {group=Wrap_ShinyStar_Uvs, parameterName=ShinyS_Rotation_Vehicle_Side, sortPriority=150}
   Alpha=@n88.G
  )
  NamedRerouteDeclaration(None=LinearInterpolate(
    A=LinearInterpolate(A=@n87, B=@n89, Alpha=@n88.R)
    B=ScalarParameter() {group=Wrap_ShinyStar_Uvs, parameterName="ShinyS_Rotation_Mech\\Chracter\\Back", sortPriority=150}
    Alpha=@n88.B
   )) {name=UVs_Star_Rotation, nodeColor=<rgba 0 0.7188 1 1>}
  n90 = ScalarParameter() {defaultValue=0.4, group=Wrap_ShinyStar, parameterName=ShinyS_Rougness}
  n91 = ScalarParameter() {defaultValue=0.5, group=Wrap_ShinyStar, parameterName=ShinyS_Rougness_Details}
  n92 = Saturate(None=@n24)
  n93 = Saturate(None=@n23)
  n94 = ScalarParameter() {defaultValue=0.15, group=Wrap_ShinyStar, parameterName=ShinyS_Spec}
  n95 = ScalarParameter() {defaultValue=0.5, group=Wrap_ShinyStar, parameterName=ShinyS_SpecDetails}
  n96 = NamedRerouteUsage()
  n97 = LinearInterpolate(A=@n96, Alpha=@n23) {constB=0}
  n98 = ScalarParameter() {defaultValue=0.3, group=Wrap_ShinyStar, parameterName=ShinyS_Emis_Boost}
  n99 = Saturate(None=Add(A=@n23, B=NamedRerouteUsage()))
  n100 = NamedRerouteUsage()
  n101 = SetMaterialAttributes!(
   MaterialAttributes=NamedRerouteUsage()
   BaseColor=LinearInterpolate(A=@n96, B=@n22.RGB, Alpha=@n23)
   Roughness=LinearInterpolate(
    A=LinearInterpolate(A=@n90, B=@n91, Alpha=@n92)
    B=ScalarParameter() {defaultValue=0.3, group=Wrap_ShinyStar, parameterName=ShinyS_Rougness_Star}
    Alpha=@n93
   )
   Specular=LinearInterpolate(A=LinearInterpolate(A=@n94, B=@n95, Alpha=@n92), B=@n95, Alpha=@n93)
   Metallic=ScalarParameter() {group=Wrap_ShinyStar, parameterName=ShinyS_Metalic}
   Normal=Add(A=@n21.Raw_Normal, B=NamedRerouteUsage())
   EmissiveColor=Add(A=Multiply(A=@n97, B=@n98), B=Multiply(A=@n99, B=@n100))
  ) {attributeSetTypes [
    #1
    #2
    #3
    #4
    #5
    #6
   ]}
  NamedRerouteDeclaration(None=@n101) {name=Mat_Base, nodeColor=<rgba 0 0.3906 1 1>}
  n102 = VectorParameter() {group=Wrap_ShinyStar_Uvs, parameterName=ShinyS_Offset_Scale_Weapon, defaultValue=<rgba -0.36 -0.01 0 0.16>}
  n103 = VectorParameter() {group=Wrap_ShinyStar_Uvs, parameterName=ShinyS_Offset_Scale_Vehicle, defaultValue=<rgba 0 0 0 1>}
  n104 = AppendVector(A=@n103.RGB, B=@n103.RGB)
  n105 = VectorParameter() {group=Wrap_ShinyStar_Uvs, parameterName=ShinyS_Offset_Scale_VehicleSide, defaultValue=<rgba 0 0 0 1>}
  n106 = AppendVector(A=@n105.RGB, B=@n105.RGB)
  n107 = MaterialFunctionCall(Float4=NamedRerouteUsage()) {materialFunction=&BreakOutFloat4Components}
  n108 = VectorParameter() {group=Wrap_ShinyStar_Uvs, parameterName=ShinyS_Offset_Scale_Mech, defaultValue=<rgba 0 0 0 1>}
  n109 = VectorParameter() {group=Wrap_ShinyStar_Uvs, parameterName=ShinyS_Offset_Scale_Character, defaultValue=<rgba 0 0 0 1>}
  n110 = LinearInterpolate(
   A=LinearInterpolate(A=AppendVector(A=@n102.RGB, B=@n102.RGB), B=LinearInterpolate(A=@n104, B=@n106, Alpha=@n107.G), Alpha=@n107.R)
   B=LinearInterpolate(A=AppendVector(A=@n108.RGB, B=@n108.RGB), B=AppendVector(A=@n109.RGB, B=@n109.RGB), Alpha=@n107.A)
   Alpha=@n107.B
  )
  n111 = VectorParameter() {group=Wrap_ShinyStar_Uvs, parameterName=ShinyS_Offset_Scale_Backpack, defaultValue=<rgba 0 0 0 1>}
  n112 = AppendVector(A=@n111.RGB, B=@n111.RGB)
  n113 = MaterialFunctionCall(Float2=NamedRerouteUsage()) {materialFunction=&BreakOutFloat2Components}
  n114 = VectorParameter() {group=Wrap_ShinyStar_Uvs, parameterName=ShinyS_Offset_Scale_Pickaxe, defaultValue=<rgba 0 0 0 1>}
  NamedRerouteDeclaration(None=LinearInterpolate(A=LinearInterpolate(A=@n110, B=@n112, Alpha=@n113.R), B=AppendVector(A=@n114.RGB, B=@n114.RGB), Alpha=@n113.G)) {name=UVs_Star_Offset_Scale, nodeColor=<rgba 0 0.625 1 1>}
  NamedRerouteDeclaration(None=TextureObjectParameter() {group=Wrap_ShinyStar, parameterName=ShinyS__Tex_BC, texture=&T_BlacksRGB}) {name=Tex_BC, nodeColor=<rgba 0.5938 0 1 1>}
  n115 = Saturate(None=@n36)
  n116 = ScalarParameter() {defaultValue=5, group=Wrap_ShinyStar_Streams, parameterName=ShinyS_StramA_Contrast}
  NamedRerouteDeclaration(None=Multiply(
    A=NamedRerouteUsage()
    B=MaterialFunctionCall(In=@n115, Contrast=@n116) {materialFunction=&CheapContrast}.Result
   )) {name=StreamsA, nodeColor=<rgba 0.875 1 0 1>}
  n117 = Saturate(None=@n18)
  n118 = ScalarParameter() {defaultValue=5, group=Wrap_ShinyStar_Streams, parameterName=ShinyS_StramC_Contrast}
  NamedRerouteDeclaration(None=Multiply(
    A=NamedRerouteUsage()
    B=MaterialFunctionCall(In=@n117, Contrast=@n118) {materialFunction=&CheapContrast}.Result
   )) {name=StreamsC, nodeColor=<rgba 0.875 1 0 1>}
  n119 = VectorParameter() {group=Wrap_ShinyStar_Uvs, parameterName=ShinyS_ScaleBK_Weapon, defaultValue=<rgba 1 1 1 1>}
  n120 = VectorParameter() {group=Wrap_ShinyStar_Uvs, parameterName=ShinyS_ScaleBK_Vehicle, defaultValue=<rgba 1 1 1 1>}
  n121 = Multiply(A=@n120.RGB, B=@n120.RGB)
  n122 = VectorParameter() {group=Wrap_ShinyStar_Uvs, parameterName=ShinyS_ScaleBK_VehicleSide, defaultValue=<rgba 1 1 1 1>}
  n123 = Multiply(A=@n122.RGB, B=@n122.RGB)
  n124 = MaterialFunctionCall(Float4=NamedRerouteUsage()) {materialFunction=&BreakOutFloat4Components}
  n125 = VectorParameter() {group=Wrap_ShinyStar_Uvs, parameterName=ShinyS_ScaleBK_Mech, defaultValue=<rgba 1 1 1 1>}
  n126 = VectorParameter() {group=Wrap_ShinyStar_Uvs, parameterName=ShinyS_ScaleBK_Character, defaultValue=<rgba 1 1 1 1>}
  n127 = LinearInterpolate(
   A=LinearInterpolate(A=Multiply(A=@n119.RGB, B=@n119.RGB), B=LinearInterpolate(A=@n121, B=@n123, Alpha=@n124.G), Alpha=@n124.R)
   B=LinearInterpolate(A=Multiply(A=@n125.RGB, B=@n125.RGB), B=Multiply(A=@n126.RGB, B=@n126.RGB), Alpha=@n124.A)
   Alpha=@n124.B
  )
  n128 = VectorParameter() {group=Wrap_ShinyStar_Uvs, parameterName=ShinyS_ScaleBK_Backpack, defaultValue=<rgba 1 1 1 1>}
  n129 = Multiply(A=@n128.RGB, B=@n128.RGB)
  n130 = MaterialFunctionCall(Float2=NamedRerouteUsage()) {materialFunction=&BreakOutFloat2Components}
  n131 = VectorParameter() {group=Wrap_ShinyStar_Uvs, parameterName=ShinyS_ScaleBK_Pickaxe, defaultValue=<rgba 1 1 1 1>}
  NamedRerouteDeclaration(None=LinearInterpolate(A=LinearInterpolate(A=@n127, B=@n129, Alpha=@n130.R), B=Multiply(A=@n131.RGB, B=@n131.RGB), Alpha=@n130.G)) {name=UVs_ScaleBK, nodeColor=<rgba 0 0.625 1 1>}
  n132 = MaterialFunctionCall(Color=Abs(None=NamedRerouteUsage())) {materialFunction=&SplitComponents}
  n133 = MaterialFunctionCall(Float4=NamedRerouteUsage()) {materialFunction=&BreakOutFloat4Components}
  n134 = LinearInterpolate(A=LinearInterpolate(A=@n132.R, B=LinearInterpolate(A=@n132.B, B=@n132.B, Alpha=@n133.G), Alpha=@n133.R), B=@n132.G, Alpha=@n133.B)
  n135 = ScalarParameter() {defaultValue=1, group=Wrap_ShinyStar, parameterName=ShinyS_Blend_Contrast, sortPriority=0}
  NamedRerouteDeclaration(None=LinearInterpolate(
    A=MaterialFunctionCall(In=@n134, Contrast=@n135) {materialFunction=&CheapContrast}.Result
    Alpha=@n133.A
   )) {name=Tri_Mask, nodeColor=<rgba 1 0.4688 0 1>}
  NamedRerouteDeclaration(None=LinearInterpolate(A=@n67, B=@n58.RGB, Alpha=NamedRerouteUsage())) {name=Bck_Color, nodeColor=<rgba 1 0 0.7734 1>}
  NamedRerouteDeclaration(None=@n83) {name=Warp_Uvs_1, nodeColor=<rgba 0.3359 1 0 1>}
  NamedRerouteDeclaration(None=@n57) {name=Msks_Stars, nodeColor=<rgba 0.2656 0 1 1>}
  NamedRerouteDeclaration(None=@n21.UVs) {name=Uvs_decal, nodeColor=<rgba 0 0.3906 1 1>}
  NamedRerouteDeclaration(None=MaterialFunctionCall!(UVs=@n48.UVs, Texture Size=ScalarParameter() {defaultValue=256, group=Wrap_ShinyStar, parameterName=ShinyS_Comput_mips}) {materialFunction=/Engine/Functions/Engine_MaterialFunctions01/Texturing/ComputeMipLevel}.MipLevel) {name=Mips, nodeColor=<rgba 1 0 0.3516 1>}
  NamedRerouteDeclaration(None=@n57) {name=Stars_Bck, nodeColor=<rgba 0.9922 0 1 1>}
  n136 = Time()
  n137 = Add(
   A=ScalarParameter() {defaultValue=-9999999, group=Wrap_Reactive, parameterName=TimeFrom-BP----------------, sortPriority=600}
   B=ScalarParameter() {group=Wrap_ShinyStar_Reactive, parameterName=ShinyS_React_Delay}
  )
  n138 = Saturate(None=Divide(A=Subtract(A=@n136, B=@n137), B=ScalarParameter() {defaultValue=1, group=Wrap_ShinyStar_Reactive, parameterName=ShinyS_React_Duration}))
  n139 = ScalarParameter() {defaultValue=0.1, group=Wrap_ShinyStar_Reactive, parameterName=ShinyS_React_Falloff}
  NamedRerouteDeclaration(None=LinearInterpolate(
    A=ScalarParameter() {defaultValue=1, group=Wrap_ShinyStar_Reactive, parameterName=ShinyS_React_Brightness_Start}
    B=ScalarParameter() {group=Wrap_ShinyStar_Reactive, parameterName=ShinyS_React_Brightness_End}
    Alpha=Power(Base=@n138, Exp=@n139)
   )) {name=Reactive, nodeColor=<rgba 0 1 0.0859 1>}
  NamedRerouteDeclaration(None=FunctionInput(Preview=Constant4Vector().RGBA) {bUsePreviewValueAsDefault=true, inputName=IsState, inputType=FunctionInput_Vector4, sortPriority=4}) {name=IsState, nodeColor=<rgba 0.1497 1 0.1633 1>}
  NamedRerouteDeclaration(None=FunctionInput(Preview=Constant2Vector()) {bUsePreviewValueAsDefault=true, inputName=IsState_B, inputType=FunctionInput_Vector2, sortPriority=4}) {name=IsState_B, nodeColor=<rgba 0.1497 1 0.1633 1>}
  n140 = PreSkinnedLocalBounds()
  n141 = Subtract(A=VertexInterpolator(VS=PreSkinnedPosition()).PS, B=@n140.Min)
  n142 = MaterialFunctionCall(Color=Subtract(A=@n140.Max, B=@n140.Max)) {materialFunction=&SplitComponents}
  n143 = Max(A=Max(A=@n142.R, B=@n142.R), B=@n142.B)
  NamedRerouteDeclaration(None=FunctionInput(Preview=Divide(A=@n141, B=@n143)) {bUsePreviewValueAsDefault=true, inputName=Calcated_Bounds, sortPriority=11}) {name=Caclcated_Bounds, nodeColor=<rgba 0 0.9297 1 1>}
  NamedRerouteDeclaration(None=TextureObjectParameter() {group=Wrap_ShinyStar, parameterName=ShinyS__Tex_BC1, texture=&T_BlacksRGB}) {name=Tex_BC1, nodeColor=<rgba 0.5938 0 1 1>}
  n144 = PreSkinnedNormal()
  NamedRerouteDeclaration(None=FunctionInput(Preview=VertexInterpolator(VS=@n144).PS) {bUsePreviewValueAsDefault=true, inputName=Pre-Skinned Local Normal, sortPriority=10}) {name=PreSkin_Normal, nodeColor=<rgba 0.8281 1 0 1>}
  NamedRerouteDeclaration(None=@n72.Normal) {name=Root_Normal, nodeColor=<rgba 0.4062 0 1 1>}
  Power(Base=Saturate(None=@n63.V2 Length))
  ScalarParameter() {defaultValue=-9999999, group=TopSideProjection_Emissive_Reactive, parameterName=Time_BP}
 ]}
