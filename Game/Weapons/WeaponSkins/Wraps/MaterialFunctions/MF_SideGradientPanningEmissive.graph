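# Material function for weapon skin wraps that adds a panning, noise-modulated side gradient to emissive, with weapon, vehicle and mech variants.
# The Pre-Skin_Bounds input supplies per-variant gradient axes and side-projection UVs, selected by isVehicle and isMech lerps.
# Time pans the noise texture, with optional stepping, glitched UVs and step-driven hue shift.
# Emissive is optionally multiplied by the primary mask, then blended by customization mask behind a static switch.
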MaterialFunction /Game/Weapons/WeaponSkins/Wraps/MaterialFunctions/MF_SideGradientPanningEmissive
graph {nodes [
  n1 = PreSkinnedPosition()
  n2 = FunctionInput(Preview=VertexInterpolator(VS=PreSkinnedLocalBounds().Half Extents).PS) {bUsePreviewValueAsDefault=true, inputName=Bounds}
  n3 = MaterialFunctionCall(Color=@n2) {materialFunction=&SplitComponents}
  n4 = StaticSwitchParameter(True=Max(A=Max(A=@n3.R, B=@n3.R), B=@n3.B), False=@n2) {
   defaultValue=true
   desc=uses the maximim size to set the texture to power of 2
   group=Side Projection UVS
   parameterName=SIde Gradioent Emissive Make texture Square
   sortPriority=-5
  }
  Divide(A=FunctionInput(Preview=VertexInterpolator(VS=@n1).PS) {bUsePreviewValueAsDefault=true, inputName=Local Position}, B=@n4)
  n5 = FunctionInput(Preview=MaterialFunctionCall() {materialFunction=/Game/Characters/Enemies/MaterialAttributes/MaterialAttributeTester}.Result) {inputName=Materials, inputType=FunctionInput_MaterialAttributes}
  n6 = GetMaterialAttributes(None=@n5) {attributeGetTypes [#1, #2]}
  n7 = VectorParameter() {group=Side Gradient Emissive, parameterName=Panning Emissive Color, defaultValue=<rgba 18.8043 0 20 1>}
  n8 = Multiply(A=Time(), B=ScalarParameter() {defaultValue=1, group=Side Gradient Emissive, parameterName=Side Panning Time Scale})
  n9 = Frac(None=@n8)
  n10 = ScalarParameter() {defaultValue=8, group=Side Gradient Emissive, parameterName=Step Frequency}
  n11 = Divide(A=Round(None=Multiply(A=@n9, B=@n10)), B=@n10)
  n12 = ScalarParameter() {group=Side Gradient Emissive, parameterName=Gradient Falloff Min Weapon}
  n13 = ScalarParameter() {defaultValue=1, group=Side Gradient Emissive, parameterName=Gradient Falloff Max Weapon}
  n14 = FunctionInput() {bUsePreviewValueAsDefault=true, inputName=Pre-Skin_Bounds}
  n15 = FunctionInput() {inputName=isVehicle, inputType=FunctionInput_Scalar}
  n16 = FunctionInput() {bUsePreviewValueAsDefault=true, inputName=isMech, inputType=FunctionInput_Scalar, previewValue=<1 0 0 1>}
  n17 = LinearInterpolate(
   A=LinearInterpolate(
    A=ChannelMaskParameter(None=@n14) {maskChannel=Green, parameterName=Emissive Rotation Axis, defaultValue=<rgba 0 1 0 0>}.R
    B=ChannelMaskParameter(None=@n14) {parameterName=Emissive Rotation Axis Vehicle}.R
    Alpha=@n15
   )
   B=ChannelMaskParameter(None=@n14) {maskChannel=Blue, parameterName=Emissive Rotation Axis Mech, defaultValue=<rgba 0 0 1 0>}.R
   Alpha=@n16
  )
  n18 = ScalarParameter() {group=Side Gradient Emissive, parameterName=Gradient Falloff Min Vehicle}
  n19 = ScalarParameter() {defaultValue=1, group=Side Gradient Emissive, parameterName=Gradient Falloff Max Vehicle}
  n20 = Multiply(A=LinearInterpolate(A=LinearInterpolate(A=@n12, B=@n13, Alpha=@n17), B=LinearInterpolate(A=@n18, B=@n19, Alpha=@n17), Alpha=@n15))
  n21 = ScalarParameter() {defaultValue=1, group=Side Gradient Emissive, parameterName=Gradient Offset}
  n22 = ScalarParameter() {defaultValue=1, group=Side Gradient Emissive, parameterName=Gradient Offset Mech}
  n23 = Saturate(None=LinearInterpolate(A=Multiply(A=@n20, B=@n21), B=Multiply(A=@n20, B=@n22), Alpha=@n16))
  n24 = ScalarParameter() {group=Side Gradient Emissive, parameterName=Pan Speed Stepping}
  n25 = VectorParameter() {group=Side Gradient Emissive, parameterName=Side Projection Panning speed, defaultValue=<rgba 0.1 0.1 0.1 1>}
  n26 = MaterialFunctionCall(Float3=@n25.RGB) {materialFunction=&BreakOutFloat3Components}
  n27 = MaterialFunctionCall(X=@n26.B, Y=@n26.B, Z=@n26.B) {materialFunction=&MakeFloat3}
  n28 = MaterialFunctionCall(Float3=Multiply(A=LinearInterpolate(A=@n8, B=@n11, Alpha=@n24), B=LinearInterpolate(A=@n25.RGB, B=@n27.Result, Alpha=@n16))) {materialFunction=&BreakOutFloat3Components}
  n29 = AppendVector(A=@n28.R, B=@n28.R)
  n30 = Multiply(A=@n14, B=ScalarParameter() {defaultValue=1, group=Side Gradient Emissive, parameterName=Tiling Texture Scale})
  n31 = ComponentMask(None=@n30) {g=true}
  n32 = ComponentMask(None=@n30) {b=true}
  n33 = Add(A=@n29, B=AppendVector(A=@n31, B=OneMinus(None=@n32)))
  n34 = VectorParameter() {group=Side Gradient Emissive, parameterName=WeaponPanningEmissiveTextureOffset, defaultValue=<rgba 0 0 0 1>}
  n35 = ComponentMask(None=@n34.RGB) {g=true, r=true}
  n36 = ComponentMask(None=@n30) {r=true}
  n37 = VectorParameter() {group=Side Gradient Emissive, parameterName=VehiclePanningEmissiveTextureOffset, defaultValue=<rgba 0 0 0 1>}
  n38 = ComponentMask(None=@n37.RGB) {g=true, r=true}
  n39 = Add(A=Add(A=@n29, B=AppendVector(A=@n36, B=@n31)), B=@n38)
  n40 = AppendVector(A=@n36, B=@n32)
  n41 = Add(A=Add(A=@n29, B=@n40), B=@n38)
  n42 = FunctionInput() {bUsePreviewValueAsDefault=true, inputName=Side Project Vehicle, inputType=FunctionInput_Scalar}
  n43 = VectorParameter() {group=Side Gradient Emissive, parameterName=MechPanningEmissiveTextureOffset, defaultValue=<rgba 0 0 0 1>}
  n44 = LinearInterpolate(
   A=LinearInterpolate(A=Add(A=@n33, B=@n35), B=LinearInterpolate(A=@n39, B=@n41, Alpha=@n42), Alpha=@n15)
   B=Add(A=Add(A=@n29, B=@n40), B=ComponentMask(None=@n43.RGB) {g=true, r=true})
   Alpha=@n16
  )
  n45 = LinearInterpolate(
   A=@n44
   B=Multiply(A=@n44, B=ScalarParameter() {defaultValue=1, group=Side Gradient Emissive, parameterName=Vehicle Panning Emissive Scale})
   Alpha=@n15
  )
  n46 = TextureSampleParameter2D(UVs=StaticSwitchParameter(True=MaterialFunctionCall!(UV=@n45) {materialFunction=/Game/Weapons/WeaponSkins/MaterialFunction/MF_GlitchFunction}.Glitched UV, False=@n45) {group=Side Gradient Emissive, parameterName=Use Glitch on PanningEmissiveTexture}) {
   group=Side Gradient Emissive
   parameterName=Panning Emissive Texture
   samplerType=SAMPLERTYPE_LinearGrayscale
   texture=/Game/Weapons/WeaponSkins/Wraps/Textures/TilingNoise06_Wrap
  }
  n47 = Multiply(
   A=LinearInterpolate(
    A=@n7.RGB
    B=MaterialFunctionCall(Hue Shift Percentage=@n11, Texture=@n7.RGB) {materialFunction=&HueShift}.Result
    Alpha=ScalarParameter() {group=Side Gradient Emissive, parameterName=Hue Shift Based on Step Count}
   )
   B=LinearInterpolate(
    A=@n23
    B=Multiply(A=@n23, B=@n46.G)
    Alpha=ScalarParameter() {defaultValue=1, group=Side Gradient Emissive, parameterName=Gradient Noise Blend}
   )
  )
  n48 = Multiply(A=ComponentMask(None=FunctionInput() {bUsePreviewValueAsDefault=true, inputName=Primary Mask}) {r=true}, B=@n47)
  n49 = SetMaterialAttributes!(
   MaterialAttributes=@n6.MaterialAttributes
   EmissiveColor=Add(
    A=@n6.EmissiveColor
    B=StaticSwitchParameter(True=@n48, False=@n47) {group=Side Gradient Emissive, parameterName="Multiply by Primary Triplaner Mask (R)"}
   )
  ) {attributeSetTypes [#1]}
  n50 = FunctionInput() {bUsePreviewValueAsDefault=true, inputName=CUstomization Mask, inputType=FunctionInput_Scalar}
  FunctionOutput(None=StaticSwitchParameter(True=BlendMaterialAttributes(A=@n5, B=@n49, Alpha=@n50), False=@n5) {group=Customization, parameterName=Use Side Gradient Panning Emissive})
 ]}
