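# Material function that layers an animated smoky glow onto weapon-skin wraps, with settings varying by item type (weapon, vehicle, mech, character, backpack).
# Pre-skinned position over local bounds forms a gradient, eroded by panning triplanar noise warped by a second distortion noise.
# The result drives curve-atlas emissive, multiplied by inverted Fresnel, and the blend strength; the raw gradient drives curve albedo.
# A customization mask limits the blend, and UseSmokeyFX statically bypasses it.
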MaterialFunction /Game/Weapons/WeaponSkins/Wraps/MaterialFunctions/MF_SmokeyFX
graph {nodes [
  n1 = ScalarParameter() {defaultValue=1, group=SmokeyFX, parameterName=SN_FresnelExponent}
  n2 = GetMaterialAttributes(None=FunctionInput() {inputName=Material, inputType=FunctionInput_MaterialAttributes}) {attributeGetTypes [#1, #2]}
  n3 = Transform(None=@n2.Normal)
  n4 = Power(
   Base=OneMinus(None=Fresnel(ExponentIn=@n1, Normal=@n3))
   Exp=ScalarParameter() {defaultValue=1, group=SmokeyFX, parameterName=SN_FresnelPower}
  )
  n5 = PreSkinnedPosition()
  n6 = FunctionInput(Preview=@n5) {bUsePreviewValueAsDefault=true, inputName=PreSkinnedPosition}
  n7 = PreSkinnedLocalBounds()
  n8 = Divide(A=@n6, B=@n7.Max)
  n9 = ComponentMask(None=@n8) {g=true}
  n10 = ComponentMask(None=@n8) {r=true}
  n11 = FunctionInput(Preview=Constant()) {bUsePreviewValueAsDefault=true, inputName=Isvehicle, inputType=FunctionInput_Scalar}
  n12 = ComponentMask(None=@n8) {g=true}
  n13 = FunctionInput(Preview=Constant()) {bUsePreviewValueAsDefault=true, inputName=IsMech, inputType=FunctionInput_Scalar}
  n14 = ComponentMask(None=@n8) {b=true}
  n15 = ScalarParameter() {defaultValue=1, group=SmokeyFX, parameterName=Character Gradiant Mult}
  n16 = ScalarParameter() {defaultValue=-1, parameterName=BackpackGradientMult}
  n17 = LinearInterpolate(
   A=LinearInterpolate(A=LinearInterpolate(A=@n9, B=@n10, Alpha=@n11), B=@n12, Alpha=@n13)
   B=LinearInterpolate(A=Multiply(A=@n14, B=@n15), B=Multiply(A=@n12, B=@n16) {desc=g}, Alpha=NamedRerouteUsage())
   Alpha=NamedRerouteUsage()
  )
  Add(A=@n4, B=@n17)
  Add(A=@n12)
  n18 = VectorParameter() {group=SmokeyFX, parameterName=SN_CharacterNoiseTiling, defaultValue=<rgba 32 32 32 1>}
  n19 = VectorParameter() {group=SmokeyFX, parameterName=SN_BackpackNoiseTiling, defaultValue=<rgba 32 32 32 1>}
  n20 = NamedRerouteUsage()
  n21 = LinearInterpolate(
   A=LinearInterpolate(
    A=VectorParameter() {group=SmokeyFX, parameterName=SN_WeaponNoiseTiling, defaultValue=<rgba 6 48 16 1>}.RGB
    B=VectorParameter() {group=SmokeyFX, parameterName=SN_VehicleNoiseTiling, defaultValue=<rgba 32 32 32 1>}.RGB
    Alpha=@n11
   )
   B=LinearInterpolate(
    A=VectorParameter() {group=SmokeyFX, parameterName=SN_MechNoiseTiling, defaultValue=<rgba 32 32 32 1>}.RGB
    B=LinearInterpolate(A=@n18.RGB, B=@n19.RGB, Alpha=@n20)
    Alpha=NamedRerouteUsage()
   )
   Alpha=@n13
  )
  n22 = LinearInterpolate(
   A=LinearInterpolate(
    A=VectorParameter() {group=SmokeyFX, parameterName=SN_WeaponDistortionTiling, defaultValue=<rgba 6 48 16 1>}.RGB
    B=VectorParameter() {group=SmokeyFX, parameterName=SN_VehicleDistortionTiling, defaultValue=<rgba 32 32 32 1>}.RGB
    Alpha=@n11
   )
   B=VectorParameter() {group=SmokeyFX, parameterName=SN_MechDistortionTiling, defaultValue=<rgba 32 32 32 1>}.RGB
   Alpha=@n13
  )
  n23 = LinearInterpolate(
   A=VectorParameter() {group=SmokeyFX, parameterName=SN_WeaponDistortionPanSpeedAndDirection, defaultValue=<rgba 0.5 0.5 0 1>}.RGB
   B=VectorParameter() {group=SmokeyFX, parameterName=SN_VehicleDistortionPanSpeedAndDirection, defaultValue=<rgba 0 -0.5 0 1>}.RGB
   Alpha=@n11
  )
  n24 = VectorParameter() {group=SmokeyFX, parameterName=SN_MechDistortionPanSpeedAndDirection, defaultValue=<rgba 0 0 -0.5 1>}
  n25 = Add(A=Divide(A=@n6, B=@n22), B=Multiply(A=LinearInterpolate(A=@n23, B=@n24.RGB, Alpha=@n13), B=Time()))
  n26 = TextureObjectParameter() {parameterName=DistortionNoise, texture=/Game/Characters/Player/Male/Medium/Bodies/M_MED_Viper/Textures/Glitch/T_SoftNoise}
  n27 = ObjectPositionWS()
  n28 = VectorParameter() {group=SmokeyFX, parameterName=BackpackCenterOffset}
  n29 = Normalize(VectorInput=Subtract(A=WorldPosition().XYZ, B=Add(A=@n27, B=@n28.RGB)))
  n30 = LinearInterpolate(A=VertexInterpolator(VS=PreSkinnedNormal()).PS, B=@n29, Alpha=NamedRerouteUsage())
  n31 = Abs(None=ComponentMask(None=@n30) {g=true})
  n32 = LinearInterpolate(
   A=TextureSample(UVs=ComponentMask(None=@n25) {g=true, r=true}, Tex=@n26).RGB
   B=TextureSample(UVs=ComponentMask(None=@n25) {b=true, r=true}, Tex=@n26).RGB
   Alpha=@n31
  )
  n33 = TextureSample(UVs=ComponentMask(None=@n25) {b=true, g=true}, Tex=@n26)
  n34 = Abs(None=ComponentMask(None=@n30) {r=true})
  n35 = Subtract(A=Multiply(A=LinearInterpolate(A=@n32, B=@n33.RGB, Alpha=@n34)) {constB=2})
  n36 = LinearInterpolate(
   A=LinearInterpolate(
    A=VectorParameter() {group=SmokeyFX, parameterName=SN_WeaponDistortionStrength, defaultValue=<rgba 0.25 0.25 0.25 1>}.RGB
    B=VectorParameter() {group=SmokeyFX, parameterName=SN_VehicleDistortionStrength, defaultValue=<rgba 0.25 0.25 0.25 1>}.RGB
    Alpha=@n11
   )
   B=VectorParameter() {group=SmokeyFX, parameterName=SN_MechDistortionStrength, defaultValue=<rgba 0.25 0.25 0.25 1>}.RGB
   Alpha=@n13
  )
  n37 = LinearInterpolate(
   A=VectorParameter() {group=SmokeyFX, parameterName=SN_WeaponNoisePanSpeedAndDirection, defaultValue=<rgba 1.5 1.5 0 1>}.RGB
   B=VectorParameter() {group=SmokeyFX, parameterName=SN_VehicleNoisePanSpeedAndDirection, defaultValue=<rgba 0 -1 0 1>}.RGB
   Alpha=@n11
  )
  n38 = VectorParameter() {group=SmokeyFX, parameterName=SN_MechNoisePanSpeedAndDirection, defaultValue=<rgba 0 0 -1 1>}
  n39 = Add(A=Add(A=Divide(A=@n6, B=@n21), B=Multiply(A=@n35, B=@n36)), B=Multiply(A=LinearInterpolate(A=@n37, B=@n38.RGB, Alpha=@n13), B=Time()))
  n40 = ComponentMask(None=@n39) {g=true, r=true}
  AppendVector(A=@n40, B=Constant())
  n41 = ObjectRadius()
  n42 = Multiply(
   A=MaterialFunctionCall(X=@n41, Y=@n41, Z=@n41) {materialFunction=&MakeFloat3}.Result
   B=ScalarParameter() {defaultValue=2, group=SmokeyFX, parameterName=ObjectBoundsBasedScaleMultiplier}
  )
  Divide(A=@n42, B=@n22)
  Divide(A=@n42, B=@n21)
  FunctionInput(Preview=@n7.Max) {bUsePreviewValueAsDefault=true, inputName=PreskinnedBounds}
  n43 = CurveAtlasRowParameter(CurveTime=@n17) {
   atlas=&AthenaCurveAtlas01
   curve=/Game/Weapons/WeaponSkins/Wraps/Textures/SkeletonHunter/SkeletonhunterwrapColor
   defaultValue=143
   parameterName=AlbedoColorCurve
  }
  n44 = ScalarParameter() {defaultValue=1, parameterName=UseAlbedoCurveforColor}
  n45 = FunctionInput(Preview=Constant3Vector()) {bUsePreviewValueAsDefault=true, inputName=Gmap}
  n46 = LinearInterpolate(A=@n17, B=Add(A=ComponentMask(None=@n45) {r=true}, B=@n17), Alpha=ScalarParameter() {group=SmokeyFX, parameterName=AddGMapEmissve})
  n47 = TextureObjectParameter() {parameterName=NoiseTexture, texture=&T_SoftNoise}
  n48 = LinearInterpolate(A=TextureSample(UVs=@n40, Tex=@n47).RGB, B=TextureSample(UVs=ComponentMask(None=@n39) {b=true, r=true}, Tex=@n47).RGB, Alpha=@n31)
  n49 = TextureSample(UVs=ComponentMask(None=@n39) {b=true, g=true}, Tex=@n47)
  n50 = ScalarParameter() {defaultValue=1, group=SmokeyFX, parameterName=SN_WeaponSmokeNoiseStrength}
  n51 = ScalarParameter() {defaultValue=1, group=SmokeyFX, parameterName=SN_VehicleSmokeNoiseStrength}
  n52 = Multiply(
   A=ComponentMask(None=LinearInterpolate(A=@n48, B=@n49.RGB, Alpha=@n34)) {r=true}
   B=LinearInterpolate(
    A=LinearInterpolate(A=@n50, B=@n51, Alpha=@n11)
    B=ScalarParameter() {defaultValue=1, group=SmokeyFX, parameterName=SN_MechSmokeNoiseStrength}
    Alpha=@n13
   )
  )
  n53 = Divide(A=Clamp(None=Subtract(A=@n46, B=@n52)), B=ScalarParameter() {defaultValue=1, group=SmokeyFX, parameterName=GradientSmoothing})
  n54 = CurveAtlasRowParameter(CurveTime=@n53) {
   atlas=&AthenaCurveAtlas01
   curve=/Game/Weapons/WeaponSkins/Wraps/Textures/SkeletonHunter/SkeletonHunterWrapGlowColor
   defaultValue=146
   group=SmokeyFX
   parameterName=EmissveColorCurve
  }
  n55 = ScalarParameter() {defaultValue=2, group=SmokeyFX, parameterName=MaterialBlendStrength}
  n56 = BlendMaterialAttributes(
   A=@n2.MaterialAttributes
   B=SetMaterialAttributes!(
    MaterialAttributes=@n2.MaterialAttributes
    BaseColor=LinearInterpolate(A=@n2.BaseColor, B=@n43, Alpha=@n44)
    EmissiveColor=Multiply(A=@n4, B=@n54)
   ) {attributeSetTypes [#1, #3]}
   Alpha=Clamp(None=Multiply(A=@n53, B=@n55))
  ) {vertexAttributeBlendType=UseA}
  n57 = ChannelMaskParameter(None=FunctionInput(Preview=Constant3Vector()) {bUsePreviewValueAsDefault=true, inputName=Mask}) {group=SmokeyFX, parameterName=CustomizationMask}
  FunctionOutput(None=StaticSwitchParameter(True=BlendMaterialAttributes(A=@n2.MaterialAttributes, B=@n56, Alpha=@n57.R), False=@n2.MaterialAttributes) {group=Customization, parameterName=UseSmokeyFX})
  MakeMaterialAttributes!(
   BaseColor=@n29
   Metallic
   Specular
   Roughness
   Anisotropy
   EmissiveColor=@n29
   Opacity
   OpacityMask
   Normal
   Tangent
   WorldPositionOffset
   SubsurfaceColor
   ClearCoat
   ClearCoatRoughness
   AmbientOcclusion
   Refraction
   CustomizedUVs_0
   CustomizedUVs_1
   CustomizedUVs_2
   CustomizedUVs_3
   CustomizedUVs_4
   CustomizedUVs_5
   CustomizedUVs_6
   CustomizedUVs_7
   PixelDepthOffset
   ShadingModel
   Displacement
  )
  n58 = VectorParameter() {group=SmokeyFX, parameterName=SN_WeaponGradientAdjustment, defaultValue=<rgba 0 0 0 1>}
  n59 = VectorParameter() {group=SmokeyFX, parameterName=SN_VehicleGradientAdjustment, defaultValue=<rgba 0 0 0 1>}
  n60 = LinearInterpolate(
   A=ScalarParameter() {group=SmokeyFX, parameterName=WeaponObjectBoundsBasedAdjustmentMultiplier}
   B=ScalarParameter() {group=SmokeyFX, parameterName=VehicleObjectBoundsBasedAdjustmentMultiplier}
   Alpha=@n11
  )
  n61 = ScalarParameter() {group=SmokeyFX, parameterName=MechObjectBoundsBasedAdjustmentMultiplier}
  Multiply(
   A=LinearInterpolate(
    A=LinearInterpolate(A=@n58.RGB, B=@n59.RGB, Alpha=@n11)
    B=VectorParameter() {group=SmokeyFX, parameterName=SN_MechGradientAdjustment, defaultValue=<rgba 0 0 0 1>}.RGB
    Alpha=@n13
   )
   B=Multiply(A=LinearInterpolate(A=@n60, B=@n61, Alpha=@n13), B=@n41)
  )
  NamedRerouteDeclaration(None=FunctionInput() {bUsePreviewValueAsDefault=true, inputName=IsCharacter, inputType=FunctionInput_Scalar}) {name=IsCharacter, nodeColor=<rgba 0 0.9253 1 1>}
  n62 = FunctionInput() {bUsePreviewValueAsDefault=true, inputName=IsStatsB, inputType=FunctionInput_Vector2}
  NamedRerouteDeclaration(None=ComponentMask(None=@n62) {r=true}) {name=IsBackpack, nodeColor=<rgba 0 0.9253 1 1>}
  NamedRerouteDeclaration(None=ComponentMask(None=@n62) {g=true}) {name=IsPickaxe, nodeColor=<rgba 0 0.9253 1 1>}
  Power(Base=DotProduct(A=CameraVectorWS(), B=VertexNormalWS()), Exp=@n1)
  VertexInterpolator(VS=@n5)
  VertexNormalWS()
 ]}
