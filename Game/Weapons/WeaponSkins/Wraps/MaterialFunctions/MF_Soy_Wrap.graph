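# Material function for the Soy weapon wrap, replacing incoming material attributes with a tinted, metallic thin-film look when the Wrap_Soy static switch is on.
# A two-color gradient along pre-skinned local Z uses separate scaling, offset and power for weapons and vehicles, plus a backpack offset.
# The result passes through ThinFilmFunction, then adds Fresnel-driven emissive based on the input normal and gradient color.
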MaterialFunction /Game/Weapons/WeaponSkins/Wraps/MaterialFunctions/MF_Soy_Wrap
graph {nodes [
  n1 = FunctionInput() {inputName=Material, inputType=FunctionInput_MaterialAttributes, sortPriority=-10}
  n2 = GetMaterialAttributes(None=@n1) {attributeGetTypes [#1, #2, #3]}
  n3 = ScalarParameter() {defaultValue=40, group=Soy, parameterName=Soy_WeaponDivisor, sortPriority=-40}
  n4 = Multiply(
   A=Divide(
    A=ObjectRadius()
    B=MaterialFunctionCall() {materialFunction=/Engine/Functions/Engine_MaterialFunctions02/WorldPositionOffset/ObjectScale}.Scale X
   )
   B=ScalarParameter() {defaultValue=0.1, group=Soy, parameterName=Soy_ObjectRadiusScale, sortPriority=-40}
  )
  n5 = FunctionInput(Preview=Constant()) {inputName=isVehicle, inputType=FunctionInput_Scalar}
  n6 = VertexInterpolator(VS=PreSkinnedPosition())
  n7 = ConstantBiasScale(None=Divide(
    A=FunctionInput(Preview=@n6.PS) {bUsePreviewValueAsDefault=true, inputName=Pre-Skinned Local Position, sortPriority=41}
    B=LinearInterpolate(A=@n3, B=@n4, Alpha=@n5)
   ))
  n8 = LinearInterpolate(
   A=ScalarParameter() {group=Soy, parameterName=Soy_WeaponGradientOffset_Z, sortPriority=30}
   B=ScalarParameter() {group=Soy, parameterName=Soy_VehicleGradientOffset_Z, sortPriority=30}
   Alpha=@n5
  )
  n9 = ScalarParameter() {group=Soy, parameterName=Soy_BackpackGradientOffset_Z, sortPriority=30}
  n10 = NamedRerouteUsage()
  n11 = Power(
   Base=Add(A=ComponentMask(None=@n7) {b=true}, B=LinearInterpolate(A=@n8, B=@n9, Alpha=@n10))
   Exp=LinearInterpolate(
    A=ScalarParameter() {defaultValue=1, group=Soy, parameterName=Soy_WeaponGradientPower_Z, sortPriority=30}
    B=ScalarParameter() {defaultValue=1, group=Soy, parameterName=Soy_VehicleGradientPower_Z, sortPriority=30}
    Alpha=@n5
   )
  )
  n12 = SetMaterialAttributes!(
   MaterialAttributes=@n2.MaterialAttributes
   BaseColor=LinearInterpolate(
    A=VectorParameter() {group=Soy, parameterName=SOY_BaseColorTint, sortPriority=-5, defaultValue=<rgba 0 0 0 1>}.RGB
    B=VectorParameter() {group=Soy, parameterName=SOY_BaseUpperColorTint, sortPriority=31, defaultValue=<rgba 0 0 0 1>}.RGB
    Alpha=Saturate(None=@n11)
   )
   Roughness=ScalarParameter() {defaultValue=0.2, group=Soy, parameterName=SOY_Roughness, sortPriority=-2}
   Normal=Constant3Vector() {constant=<rgba 0 0 1 0>}
   Specular=ScalarParameter() {defaultValue=1, group=Soy, parameterName=SOY_Specular, sortPriority=-1}
   Metallic=ScalarParameter() {defaultValue=1, group=Soy, parameterName=SOY_Metallic, sortPriority=-3}
  ) {attributeSetTypes [
    #1
    #2
    #3
    #4
    #5
   ]}
  n13 = GetMaterialAttributes(None=MaterialFunctionCall!(Alpha=Constant() {r=1}, Material Attributes=@n12) {materialFunction=/Game/Weapons/WeaponSkins/MaterialFunction/ThinFilmFunction}.Result) {attributeGetTypes [#6, #1]}
  n14 = Multiply(A=@n13.BaseColor, B=ScalarParameter() {defaultValue=1.25, group=Soy, parameterName=Fresnel_Basecolor_Mult})
  n15 = ScalarParameter() {defaultValue=6, group=Soy, parameterName=FresnelExponent, sortPriority=0}
  n16 = ScalarParameter() {group=Soy, parameterName=BaseReflectFraction, sortPriority=1}
  n17 = Transform(None=@n2.Normal)
  n18 = Saturate(None=Multiply(
    A=GetMaterialAttributes(None=@n12) {attributeGetTypes [#1]}.BaseColor
    B=Fresnel(ExponentIn=@n15, BaseReflectFractionIn=@n16, Normal=@n17) {exponent=0.3}
   ))
  n19 = SetMaterialAttributes!(
   MaterialAttributes=@n13.MaterialAttributes
   EmissiveColor=Add(A=@n13.EmissiveColor, B=Multiply(A=@n14, B=@n18))
   BaseColor=@n13.MaterialAttributes
  ) {attributeSetTypes [#6, #1]}
  n20 = Constant() {r=1}
  FunctionOutput(None=StaticSwitchParameter(True=BlendMaterialAttributes(A=@n2.MaterialAttributes, B=@n19, Alpha=@n20), False=@n1) {desc=Simple Material and thin film, group=Customization_Wrap, parameterName=Wrap_Soy})
  n21 = FunctionInput() {bUsePreviewValueAsDefault=true, inputName=IsStatsB, inputType=FunctionInput_Vector2, sortPriority=78}
  NamedRerouteDeclaration(None=ComponentMask(None=NamedRerouteDeclaration(None=@n21) {name=Is_StatsB, nodeColor=<rgba 0 0.9253 1 1>}) {r=true}) {name=IsBackpack, nodeColor=<rgba 0.2656 0 1 1>}
 ]}
