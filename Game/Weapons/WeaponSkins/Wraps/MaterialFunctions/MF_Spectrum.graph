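# Wrap material function that adds a switchable "Spectrum" skin effect to weapons, vehicles and mechs, blended into incoming material attributes by the primary customization mask.
# Wrap bounds drive side-projected masks and an animated gradient, warped by a top/side projected mask.
# That gradient, shaped by curvature pits and peaks, indexes a time-scrolled color ramp for emissive.
# The side mask splits emissive color areas from dark areas, setting roughness and where projected normal detail applies.
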MaterialFunction /Game/Weapons/WeaponSkins/Wraps/MaterialFunctions/MF_Spectrum
graph {nodes [
  n1 = FunctionInput() {inputName=Material, inputType=FunctionInput_MaterialAttributes, sortPriority=-200}
  n2 = GetMaterialAttributes(None=@n1) {attributeGetTypes [#1, #2, #3]}
  n3 = ScalarParameter() {group=Spectrum, parameterName=SP_Roughness_Black_low}
  n4 = ScalarParameter() {defaultValue=1, group=Spectrum, parameterName=SP_Roughness_Black_high}
  n5 = VectorParameter() {group=Spectrum, parameterName=SP_Side_Mask_Offset_vehicle, sortPriority=5, defaultValue=<rgba 0 0 0 1>}
  n6 = VectorParameter() {group=Spectrum, parameterName=SP_Side_Mask_Offset_vehicleSP, sortPriority=5, defaultValue=<rgba 0 0 0 1>}
  n7 = Constant()
  n8 = FunctionInput(Preview=@n7) {bUsePreviewValueAsDefault=true, inputName=Side Project Vehicle, inputType=FunctionInput_Scalar}
  n9 = FunctionInput(Preview=Constant()) {bUsePreviewValueAsDefault=true, inputName=isVehicle, inputType=FunctionInput_Scalar, sortPriority=40}
  n10 = LinearInterpolate(
   A=VectorParameter() {group=Spectrum, parameterName=SP_Side_Mask_Offset_Weapon, sortPriority=5}.RGB
   B=LinearInterpolate(A=@n5.RGB, B=@n6.RGB, Alpha=@n8)
   Alpha=@n9
  )
  n11 = VectorParameter() {group=Spectrum, parameterName=SP_Side_Mask_Offset_Mech, sortPriority=5, defaultValue=<rgba 0 0 0 1>}
  n12 = FunctionInput(Preview=@n7) {bUsePreviewValueAsDefault=true, inputName=isMech, inputType=FunctionInput_Scalar, sortPriority=41}
  n13 = PreSkinnedPosition()
  n14 = MaterialFunctionCall!(
   is Vehicle=@n9
   Object Bounds Max Adjustment=FunctionInput(Preview=Constant3Vector() {constant=<rgba 0 -0.5 0 1>}) {inputName=Object Bounds Max Adjustment, sortPriority=52}
   Object Bounds Min Adjustment=FunctionInput(Preview=Constant3Vector() {constant=<rgba 0 31.5 0 1>}) {inputName=Object Bounds Min Adjustment, sortPriority=51}
   Object Bounds Center Adjustment=FunctionInput(Preview=Constant3Vector() {constant=<rgba 0.2122 0 -0.09 0>}) {inputName=Object Bounds Center Adjustment, sortPriority=50}
   isMech=@n12
   Side Project Vehicle
   Local Position=FunctionInput(Preview=VertexInterpolator(VS=@n13).PS) {bUsePreviewValueAsDefault=true, inputName=Local Position}
  ) {materialFunction=&/MF_WrapBounds}
  n15 = Constant3Vector() {constant=<rgba 1 1 -1 0>}
  n16 = MaterialFunctionCall(Color=Add(A=LinearInterpolate(A=@n10, B=@n11.RGB, Alpha=@n12), B=Multiply(A=@n14.Result, B=@n15))) {materialFunction=&SplitComponents}
  n17 = AppendVector(A=@n16.G, B=@n16.G)
  n18 = LinearInterpolate(A=AppendVector(A=@n16.R, B=@n16.R), B=AppendVector(A=@n16.R, B=@n16.R), Alpha=@n8)
  n19 = OneMinus(None=@n16.B)
  n20 = TextureSample(
   UVs=LinearInterpolate(A=LinearInterpolate(A=@n17, B=@n18, Alpha=@n9), B=AppendVector(A=@n19, B=@n16.R), Alpha=@n12)
   Tex=TextureObjectParameter() {group=Spectrum, parameterName=SP_Side_Mask, sortPriority=1, texture=&T_BlacksRGB}
  )
  n21 = ChannelMaskParameter(None=@n20.RGBA) {group=Spectrum, maskChannel=Blue, parameterName=SP_Side_Mask_01_Roghness, defaultValue=<rgba 0 0 1 0>}
  n22 = ChannelMaskParameter(None=@n20.RGBA) {group=Spectrum, parameterName=SP_Side_Mask_Channel, sortPriority=2}
  n23 = VectorParameter() {group=Spectrum, parameterName=SP_Grad_OffsetWeapon, sortPriority=60, defaultValue=<rgba 0 -1 0 0>}
  n24 = VectorParameter() {group=Spectrum, parameterName=SP_Grad_OffsetVehicle, sortPriority=60, defaultValue=<rgba -1 0 0 0>}
  n25 = VectorParameter() {group=Spectrum, parameterName=SP_Grad_OffsetMech, sortPriority=60, defaultValue=<rgba 0 0 -1 0>}
  n26 = StaticBool()
  n27 = LinearInterpolate(
   A=LinearInterpolate(
    A=ScalarParameter() {defaultValue=0.5, group=Spectrum, parameterName=SP_Grad_Spread_Weapon, sortPriority=70}
    B=ScalarParameter() {defaultValue=0.5, group=Spectrum, parameterName=SP_Grad_Spread_Vehicle, sortPriority=70}
    Alpha=@n9
   )
   B=ScalarParameter() {defaultValue=0.5, group=Spectrum, parameterName=SP_Grad_Spread_Mech, sortPriority=70}
   Alpha=@n12
  )
  n28 = VectorParameter() {group=Spectrum, parameterName=SP_Grad_UV_Animated_XY, sortPriority=70}
  n29 = TextureObjectParameter() {
   group=Spectrum
   parameterName=SP_TopSide_Mask_01
   samplerType=SAMPLERTYPE_Masks
   sortPriority=100
   texture=&T_BlackMask
  }
  n30 = LinearInterpolate(
   A=ScalarParameter() {defaultValue=1, group=Spectrum, parameterName=SP_TopSide_Blend_Contrast, sortPriority=110}
   B=ScalarParameter() {defaultValue=2, group=Spectrum, parameterName=SP_TopSide_Blend_Contrast_Mech, sortPriority=110}
   Alpha=@n12
  )
  n31 = ScalarParameter() {group=Spectrum, parameterName=SP_TopSide_Fill_alternate_Axis_Opacity 01, sortPriority=104}
  n32 = VectorParameter() {group=Spectrum, parameterName=SP_TopSide_Fill_alternate_Axis_Color_Mask_01, sortPriority=105, defaultValue=<rgba 0 0 0 1>}
  n33 = AppendVector(A=@n32.RGB, B=@n32.RGB)
  n34 = FunctionInput(Preview=VertexInterpolator(VS=PreSkinnedNormal()).PS) {bUsePreviewValueAsDefault=true, inputName=Local Normals}
  n35 = LinearInterpolate(
   A=LinearInterpolate(A=Constant3Vector() {constant=<rgba 0 0 -0.25 1>}, B=Constant3Vector() {constant=<rgba 0 0 0 1>}, Alpha=@n9)
   B=Constant3Vector() {constant=<rgba 0.25 0.25 0 1>}
   Alpha=@n12
  )
  n36 = VectorParameter() {group=Spectrum, parameterName="SP_TopSide_Scale_Weapon_01 A=uniform", sortPriority=120, defaultValue=<rgba 1 1 1 1>}
  n37 = VectorParameter() {group=Spectrum, parameterName="SP_TopSide_Scale_Vehicle_01 A=uniform", sortPriority=120, defaultValue=<rgba 1 1 1 1>}
  n38 = LinearInterpolate(A=Multiply(A=@n36.RGB, B=@n36.RGB), B=Multiply(A=@n37.RGB, B=@n37.RGB), Alpha=@n9)
  n39 = VectorParameter() {group=Spectrum, parameterName="SP_TopSide_Scale_Mech_01 A=uniform", sortPriority=120, defaultValue=<rgba 1 1 1 1>}
  n40 = Multiply(A=@n39.RGB, B=@n39.RGB)
  n41 = VectorParameter() {group=Spectrum, parameterName=SP_TopSide_Panning_Weapon_01---------------------, sortPriority=125, defaultValue=<rgba 0 0 0 1>}
  n42 = VectorParameter() {group=Spectrum, parameterName=SP_TopSide_Panning_Vehicle_01, sortPriority=125, defaultValue=<rgba 0 0 0 1>}
  n43 = LinearInterpolate(A=@n41.RGB, B=@n42.RGB, Alpha=@n9)
  n44 = VectorParameter() {group=Spectrum, parameterName=SP_TopSide_Panning_Mech_01, sortPriority=125, defaultValue=<rgba 0 0 0 1>}
  n45 = Add(
   A=Multiply(A=@n14.Result, B=LinearInterpolate(A=@n38, B=@n40, Alpha=@n12))
   B=Multiply(A=Time(), B=LinearInterpolate(A=@n43, B=@n44.RGB, Alpha=@n12))
  )
  n46 = Multiply(
   A=ChannelMaskParameter(None=MaterialFunctionCall!(
     IsCharacter
     IsStatsB
     Texture=@n29
     Rotation_Axis=@n35
     Caculated_Bounds=@n45
     Pre_Skinned_Normals=@n34
     Blend_Contrast=@n30
     Fill_alternate_Axis_Opacity=@n31
     "Fill_alternate_Axis_Color_A=Roughness"=@n33
     Is_Vehicle=@n9
     Is_Mech=@n12
     Enable_Fill_Z_Weapon
     Fill_Z_Weapon_opacity
     Fill_Z_Weapon_Color
    ) {materialFunction=&/MF_TopsSideProjection_Base}.Combined) {group=Spectrum, parameterName=SP_TopSide_Warp_Mask_01, sortPriority=80}.R
   B=ScalarParameter() {group=Spectrum, parameterName=SP_TopSide_Warp_Amount, sortPriority=80}
  )
  n47 = ScalarParameter() {defaultValue=1, group=Spectrum, parameterName=SP_Grad_Texture_Rotate_Mech, sortPriority=75}
  n48 = ScalarParameter() {defaultValue=1, group=Spectrum, parameterName=SP_Grad_Texture_Rotate_Vehicle, sortPriority=75}
  n49 = FunctionInput(Preview=Constant3Vector() {constant=<rgba 0 1 0 0>}) {bUsePreviewValueAsDefault=true, inputName=ForwardAxis}
  n50 = TextureObjectParameter() {group=Spectrum, parameterName=SP_Gradent_Texture, sortPriority=55, texture=&T_BlacksRGB}
  n51 = ChannelMaskParameter(None=TextureSample(
    UVs=MaterialFunctionCall!(
     ForwardAxis=@n49
     Bounds=@n14.Result
     Pre-Skinned Local Position
     isVehicle=@n9
     isMech=@n12
     Side Project Vehicle
     WeaponOffset=@n23.RGB
     WeaponFalloff
     VehicleOffset=@n24.RGB
     VehicleFalloff
     MechOffset=@n25.RGB
     MechFalloff
     Invert_Gradent=@n26
     UV_Spread_XY=@n27
     UV_Animated_XY=@n28.RGB
     Warp_UVs=@n46
     Rotate_Mech=@n47
     Rotate_Vehicle_Side
     Rotate_Vehicle=@n48
    ) {materialFunction=&/MF_Wrap_Gradent_Creation}.Uvs
    Tex=@n50
   ).RGB) {group=Spectrum, parameterName=SP_Gradent_Channel, sortPriority=56}
  n52 = FunctionInput() {inputName=Customization Mask Curvatur, inputType=FunctionInput_Scalar, sortPriority=50, previewValue=<1 0 0 1>}
  n53 = MaterialFunctionCall(
   Input=@n52
   Input Low=Constant()
   Input High=Constant() {r=0.5}
   Target Low=ScalarParameter() {group=Spectrum, parameterName=Sp_Pits_low, sortPriority=54}
   Target High=ScalarParameter() {defaultValue=1, group=Spectrum, parameterName=Sp_Pits_High, sortPriority=54}
  ) {materialFunction=&RemapValueRange}
  n54 = MaterialFunctionCall(
   Input=@n52
   Input Low=Constant() {r=0.5}
   Input High=Constant() {r=1}
   Target Low=ScalarParameter() {group=Spectrum, parameterName=Sp_Peaks_low, sortPriority=54}
   Target High=ScalarParameter() {defaultValue=1, group=Spectrum, parameterName=Sp_Peaks_High, sortPriority=54}
  ) {materialFunction=&RemapValueRange}
  n55 = Multiply(A=Saturate(None=Multiply(A=@n51.R, B=@n53.Result)), B=@n54.Result)
  n56 = ChannelMaskParameter(None=@n20.RGBA) {group=Spectrum, maskChannel=Green, parameterName=SP_Side_MaskB_Channel, sortPriority=56, defaultValue=<rgba 0 1 0 0>}
  n57 = Add(
   A=Clamp(None=Multiply(A=@n55, B=@n56.R)) {maxDefault=0.99}
   B=ScalarParameter() {group=Spectrum, parameterName=SP_Gradent_Color_Shift, sortPriority=52}
  )
  n58 = Multiply(A=Time(), B=ScalarParameter() {group=Spectrum, parameterName=SP_Gradent_Color_SpeedShift, sortPriority=52})
  n59 = Constant() {r=1}
  n60 = Multiply(A=Saturate(None=Multiply(A=@n59, B=@n53.Result)), B=@n54.Result)
  n61 = Multiply(
   A=TextureSample(
    UVs=Add(A=@n57, B=@n58)
    Tex=TextureObjectParameter() {group=Spectrum, parameterName=SP_Gradent_Color, sortPriority=50, texture=&T_BlacksRGB}
   ).RGB
   B=Multiply(A=Multiply(A=@n60, B=@n56.R), B=ScalarParameter() {group=Spectrum, parameterName=SP_Gradent_Color_Brightness, sortPriority=51})
  )
  n62 = TextureObjectParameter() {
   group=Spectrum
   parameterName=SP_TopSide_Normal_01
   samplerType=SAMPLERTYPE_Normal
   sortPriority=100
   texture=&FlatNormalBlue
  }
  n63 = LinearInterpolate(
   A=MaterialFunctionCall!(
    Texture_Normals=@n62
    Rotation_Axis=@n35
    Caculated_Bounds=@n45
    Pre_Skinned_Normals=@n34
    Blend_Contrast=@n30
    Fill_alternate_Axis_Opacity=@n31
    Is_Vehicle=@n9
    Is_Mech=@n12
   ) {materialFunction=&/MF_TopsSideProjection_Normals}.Add_Surface_Normal
   B=Constant3Vector()
   Alpha=@n22.R
  )
  n64 = SetMaterialAttributes!(
   MaterialAttributes=@n2.MaterialAttributes
   BaseColor=VectorParameter() {group=Spectrum, parameterName=SP_BaseColor, defaultValue=<rgba 0 0 0 1>}.RGB
   Roughness=LinearInterpolate(
    A=ScalarParameter() {defaultValue=0.2, group=Spectrum, parameterName=SP_Roughness_ColorArea}
    B=LinearInterpolate(A=@n3, B=@n4, Alpha=@n21.R)
    Alpha=@n22.R
   )
   EmissiveColor=Multiply(A=OneMinus(None=@n22.R), B=@n61)
   Normal=LinearInterpolate(
    A=@n2.Normal
    B=Add(A=@n2.Normal, B=@n63)
    Alpha=ScalarParameter() {defaultValue=1, group=Spectrum, parameterName=SP_TopSide_Normal_01_Intensty, sortPriority=100}
   )
  ) {attributeSetTypes [
    #3
    #4
    #2
    #1
   ]}
  n65 = FunctionInput() {inputName=Customization Mask Primary, inputType=FunctionInput_Scalar, sortPriority=50, previewValue=<1 0 0 1>}
  FunctionOutput(None=StaticSwitchParameter(True=BlendMaterialAttributes(A=@n1, B=@n64, Alpha=@n65) {vertexAttributeBlendType=UseB}, False=@n1) {group=Customization_Wrap, parameterName=Wrap_Sprectrum})
  MaterialFunctionCall(Hue Shift Percentage=ScalarParameter() {group=Spectrum, parameterName=SP_Hue_Shift}, Texture=@n61) {materialFunction=&HueShift}
 ]}
