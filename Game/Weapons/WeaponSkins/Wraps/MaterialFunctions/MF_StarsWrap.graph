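# Material function for the Stars cosmetic wrap on weapons, vehicles, characters and backpacks, passing input through when the Wrap_Stars switch is off.
# Panning constellation layers project onto a camera-facing virtual plane at the object, with view-direction parallax and per-type scale or tiling.
# Crevasse-masked fresnel and animated hue shift color the layers, driving base color, emissive, roughness, specular and normal.
# Height-lerped stars add glow, boosted briefly from a Blueprint-supplied fire time.
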MaterialFunction /Game/Weapons/WeaponSkins/Wraps/MaterialFunctions/MF_StarsWrap
graph {nodes [
  n1 = PreSkinnedLocalBounds()
  Add(A=@n1.Min) {constB=0}
  Add() {constB=0}
  n2 = DotProduct(
   A=CameraVectorWS()
   B=StaticSwitchParameter(True=PixelNormalWS(), False=VertexNormalWS()) {group=Constellation, parameterName=FresnelUsesPixelNormals}
  )
  n3 = OneMinus(None=@n2)
  n4 = ScalarParameter() {group=Constellation, parameterName=CS_EmissiveFalloff}
  Clamp(None=Power(Base=Subtract(A=@n3, B=@n4) {constB=0.35}, Exp=ScalarParameter() {defaultValue=2, group=Constellation, parameterName=CS_EmissvePower}))
  n5 = MaterialFunctionCall(Float2=ViewSize()) {materialFunction=&BreakOutFloat2Components}
  n6 = AppendVector(A=Constant() {r=1}, B=Divide(A=@n5.G, B=@n5.G))
  n7 = Multiply(A=ConstantBiasScale(None=ScreenPosition().ViewportUV) {bias=-0.5, scale=2}, B=@n6)
  n8 = ActorPositionWS()
  n9 = Min(A=MaterialFunctionCall(World Position=@n8) {materialFunction=&TransformToClipSpace}.Clip Space XY)
  n10 = Constant()
  n11 = Constant()
  n12 = Multiply(
   A=Multiply(A=Constant2Vector() {g=-0.5, r=0.5}, B=Max(A=@n9) {constB=-1})
   B=If(A=Distance(A=@n8, B=@n10), B=Constant() {r=0.01}, "A > B"=Constant() {r=1}, "A == B"=@n11, "A < B"=@n11)
  )
  n13 = VertexInterpolator(VS=Subtract(A=ConstantBiasScale(None=@n7), B=Multiply(A=@n6, B=@n12)))
  ComponentMask(None=@n13.PS) {g=true, r=true}
  n14 = FunctionInput() {inputName=Material, inputType=FunctionInput_MaterialAttributes, sortPriority=-200}
  n15 = GetMaterialAttributes(None=@n14) {attributeGetTypes [
    #1
    #2
    #3
    #4
    #5
    #6
   ]}
  n16 = ObjectPositionWS()
  n17 = Normalize(VectorInput=Subtract(A=@n16, B=CameraPositionWS()))
  n18 = CrossProduct(A=@n17, B=Constant3Vector() {constant=<rgba 0 0 1 1>})
  n19 = MaterialFunctionCall(Float3=PreSkinnedLocalBounds().Half Extents) {materialFunction=&BreakOutFloat3Components}
  n20 = Multiply(
   A=MaterialFunctionCall() {materialFunction=/Engine/Functions/Engine_MaterialFunctions02/WorldPositionOffset/ObjectScale}.Scale X
   B=Max(A=Max(A=@n19.R, B=@n19.R), B=@n19.B)
  )
  n21 = VectorParameter() {group=Constellation, parameterName=ScreenCoordScaleWeapons, defaultValue=<rgba 10 10 0 0>}
  n22 = VectorParameter() {group=Constellation, parameterName=ScreenCoordScaleVehicles, defaultValue=<rgba 10 10 0 0>}
  n23 = FunctionInput() {inputName=IsVehicle, inputType=FunctionInput_Scalar, sortPriority=4}
  n24 = FunctionInput() {inputName=IsMech, inputType=FunctionInput_Scalar, sortPriority=4}
  n25 = ComponentMask(None=Saturate(None=Add(A=@n23, B=@n24))) {r=true}
  n26 = Multiply(
   A=MaterialFunctionCall!(Plane Normal Axis=@n17, Plane X Axis=Normalize(VectorInput=@n18), Plane Center=@n16, UV Scale=@n20) {materialFunction=/Game/Athena/Environments/Nexus/Materials/VirtualPlaneCoordinatesAbs}.UVs
   B=ComponentMask(None=LinearInterpolate(A=@n21.RGB, B=@n22.RGB, Alpha=@n25)) {g=true, r=true}
  )
  n27 = ScalarParameter() {defaultValue=1, group=Constellation, parameterName=Parallax0_X}
  n28 = ScalarParameter() {defaultValue=1, group=Constellation, parameterName=Parallax0_Y}
  n29 = Subtract(A=ActorPositionWS(), B=CameraPositionWS())
  n30 = Normalize(VectorInput=ComponentMask(None=@n29) {g=true, r=true})
  n31 = Constant2Vector() {g=1}
  n32 = Add(A=Multiply(A=DotProduct(A=@n30, B=@n31)) {constB=0.5}) {constB=0.5}
  n33 = Normalize(VectorInput=@n29)
  n34 = Constant3Vector() {constant=<rgba 0 0 1 0>}
  n35 = Add(A=Multiply(A=DotProduct(A=@n33, B=@n34)) {constB=0.5}) {constB=0.5}
  n36 = ComponentMask(None=VertexInterpolator(VS=AppendVector(A=@n32, B=@n35)).PS) {g=true, r=true}
  n37 = Add(A=@n26, B=Multiply(A=@n36, B=AppendVector(A=@n27, B=@n28))) {constB=0}
  n38 = LinearInterpolate(
   A=ScalarParameter() {defaultValue=0.25, group=Constellation, parameterName=CS_Layer0_Tiling}
   B=ScalarParameter() {defaultValue=0.25, group=Constellation, parameterName=CS_Layer0_Tiling_Character}
   Alpha=NamedRerouteUsage()
  )
  n39 = ScalarParameter() {defaultValue=0.25, group=Constellation, parameterName=CS_Layer0_Tiling_Backpack}
  n40 = NamedRerouteUsage()
  n41 = VectorParameter() {group=Constellation, parameterName=CS_Layer_01_PanSpeedAndDirection, defaultValue=<rgba 0.035 -0.25 0 1>}
  n42 = Panner(Coordinate=Multiply(A=@n37, B=LinearInterpolate(A=@n38, B=@n39, Alpha=@n40)), Speed=ComponentMask(None=@n41.RGB) {g=true, r=true})
  n43 = TextureObjectParameter() {
   group=Constellation
   parameterName=BaseTexture
   texture=/Game/Characters/Player/Female/Medium/Bodies/F_MED_Rebirth_Soldier/Skins/Constellation/Textures/T_Constellation_Pattern
  }
  n44 = TextureSample(UVs=@n42, Tex=@n43) {constMipValue=0, texture=/Game/Characters/Player/Female/Medium/Bodies/F_MED_Gift_Wrap/Textures/T_Snowflake_Pattern}
  n45 = TextureSample(UVs=@n42, Tex=@n43) {constMipValue=0, texture=&T_Snowflake_Pattern}
  n46 = Multiply(
   A=Clamp(
    None=ChannelMaskParameter(None=@n45.A) {group=Constellation, parameterName=CS_Layer1Transition_TextureChannel}.R
    Max=ScalarParameter() {defaultValue=1, group=Constellation, parameterName=CS_Layer0_Gradient_Clamp Max}
   ) {maxDefault=0.99, minDefault=0.01}
   B=ScalarParameter() {defaultValue=0.5, group=Constellation, parameterName=CS_Layer0_Gradient_Opacity_CE}
  )
  n47 = VectorParameter() {group=Constellation, parameterName=CS_Layer0Lerp_PanSpeedAndDirection, defaultValue=<rgba 0.035 -0.25 0 1>}
  n48 = Panner(
   Coordinate=Multiply(A=@n37, B=ScalarParameter() {defaultValue=0.1, group=Constellation, parameterName=CS_Layer0Lerp_Tiling})
   Speed=ComponentMask(None=@n47.RGB) {g=true, r=true}
  )
  n49 = TextureObjectParameter() {group=Constellation, parameterName=StarsTexture, texture=/Game/Weapons/WeaponSkins/Wraps/Textures/StarsOfficial/T_Wrap_Stars_M2}
  n50 = Saturate(None=MaterialFunctionCall(
    Transition Phase=Saturate(None=@n46)
    Height Texture=TextureSample(UVs=@n48, Tex=@n49) {constMipValue=0, texture=&T_Snowflake_Pattern}.B
    Contrast=ScalarParameter() {defaultValue=0.5, group=Constellation, parameterName=CS_Layer0_Contrast}
   ) {materialFunction=&HeightLerp}.Alpha)
  n51 = Multiply(A=@n50, B=ScalarParameter() {defaultValue=2, group=Constellation, parameterName=CS_Layer0_AmbientIntensity}) {constB=2}
  n52 = VectorParameter() {group=Constellation, parameterName=Layer1_color, defaultValue=<rgba 0.0314 3.8406 5 1>}
  n53 = Multiply(A=Add(A=@n44.G, B=Multiply(A=@n44.G, B=@n51)), B=@n52.RGB)
  n54 = Multiply(
   A=Multiply(A=@n50, B=ScalarParameter() {defaultValue=8, group=Constellation, parameterName=CS_Layer0_ReactiveIntensity}) {constB=8}
   B=@n44.G
  )
  n55 = Subtract(A=Time(), B=ScalarParameter() {group=00-Reactive-GTFB, parameterName=00-GameTimeFromBP, sortPriority=199})
  n56 = ScalarParameter() {defaultValue=0.8, group=Stars_Reactive, parameterName=CS_Time_Emissive01_Duration, sortPriority=200}
  n57 = Power(
   Base=Saturate(None=Divide(A=@n55, B=@n56))
   Exp=ScalarParameter() {defaultValue=0.8, group=Stars_Reactive, parameterName=CS_Time_Emissive01_Curve, sortPriority=201}
  )
  n58 = LinearInterpolate(A=@n57, B=OneMinus(None=@n57), Alpha=ScalarParameter() {group=Stars_Reactive, parameterName=CS_ReactiveTest, sortPriority=250})
  n59 = LinearInterpolate(A=Add(A=@n53, B=Multiply(A=@n54, B=@n52.RGB)), B=@n53, Alpha=@n58)
  n60 = ScalarParameter() {group=Constellation, parameterName=Parallax1_X}
  n61 = ScalarParameter() {group=Constellation, parameterName=Parallax1_Y}
  n62 = Add(A=@n26, B=Multiply(A=@n36, B=AppendVector(A=@n60, B=@n61))) {constB=0}
  n63 = LinearInterpolate(
   A=LinearInterpolate(
    A=ScalarParameter() {defaultValue=0.75, group=Constellation, parameterName=CS_Layer1_Tiling}
    B=ScalarParameter() {defaultValue=0.75, group=Constellation, parameterName=CS_Layer1_Tiling_Character}
    Alpha=NamedRerouteUsage()
   )
   B=ScalarParameter() {defaultValue=0.75, group=Constellation, parameterName=CS_Layer1_Tiling_Backpack}
   Alpha=NamedRerouteUsage()
  )
  n64 = VectorParameter() {group=Constellation, parameterName=CS_Layer_02_PanSpeedAndDirection, defaultValue=<rgba 0.035 -0.25 0 1>}
  n65 = TextureSample(UVs=Panner(Coordinate=Multiply(A=@n62, B=@n63), Speed=ComponentMask(None=@n64.RGB) {g=true, r=true}), Tex=@n43) {texture=&T_Snowflake_Pattern}
  n66 = VectorParameter() {group=Constellation, parameterName=Layer2_color, defaultValue=<rgba 0.0125 1.5363 2 1>}
  n67 = Multiply(
   A=@n36
   B=AppendVector(A=ScalarParameter() {group=Constellation, parameterName=Parallax2_X}, B=ScalarParameter() {group=Constellation, parameterName=Parallax2_Y})
  )
  n68 = LinearInterpolate(
   A=ScalarParameter() {defaultValue=1.25, group=Constellation, parameterName=CS_Layer2_Tiling}
   B=ScalarParameter() {defaultValue=1.25, group=Constellation, parameterName=CS_Layer2_Tiling_Character}
   Alpha=NamedRerouteUsage()
  )
  n69 = ScalarParameter() {defaultValue=1.25, group=Constellation, parameterName=CS_Layer2_Tiling_Backpack}
  n70 = NamedRerouteUsage()
  n71 = VectorParameter() {group=Constellation, parameterName=CS_Layer_03_PanSpeedAndDirection, defaultValue=<rgba 0.035 -0.25 0 1>}
  n72 = Panner(
   Coordinate=Multiply(A=Add(A=@n26, B=@n67) {constB=0}, B=LinearInterpolate(A=@n68, B=@n69, Alpha=@n70))
   Speed=ComponentMask(None=@n71.RGB) {g=true, r=true}
  )
  n73 = Max(
   A=Max(A=Max(A=@n53, B=@n59), B=Multiply(A=@n65.R, B=@n66.RGB) {constB=0.5})
   B=Multiply(
    A=TextureSample(UVs=@n72, Tex=@n43) {texture=&T_Snowflake_Pattern}.B
    B=VectorParameter() {group=Constellation, parameterName=Layer3_color, defaultValue=<rgba 0.0063 1 0.991 1>}.RGB
   ) {constB=0.25}
  )
  n74 = Desaturation(None=Multiply(A=ComponentMask(None=@n73) {r=true}) {constB=5})
  n75 = LinearInterpolate(
   A=ScalarParameter() {defaultValue=0.5, group=Constellation, parameterName=CS_SpecularityBase, sortPriority=-99}
   B=ScalarParameter() {group=Constellation, parameterName=CS_SpecularityMask, sortPriority=-99}
   Alpha=@n74
  )
  n76 = LinearInterpolate(
   A=ScalarParameter() {defaultValue=0.5, group=Constellation, parameterName=CS_RoughnessBase, sortPriority=-98}
   B=ScalarParameter() {defaultValue=0.5, group=Constellation, parameterName=CS_RoughnessMask, sortPriority=-98}
   Alpha=@n74
  )
  n77 = Subtract(A=Saturate(None=@n2)) {constB=0.1}
  n78 = Clamp(None=Multiply(A=Power(Base=@n77)) {constB=2})
  n79 = FunctionInput() {inputName="Customization Crevasse (_CM G)", inputType=FunctionInput_Scalar, sortPriority=21, previewValue=<1 0 0 1>}
  n80 = Constant()
  n81 = ScalarParameter() {defaultValue=0.5, group=Constellation, parameterName=NoCrevasseFresnelMaskValue}
  n82 = Abs(None=Add(A=If(A=@n79, B=@n80, "A > B"=@n79, "A == B"=@n81, "A < B"=@n79)) {constB=-0.5})
  n83 = Max(A=Saturate(None=Multiply(A=@n82) {constB=2}), B=ScalarParameter() {defaultValue=0.12, group=Constellation, parameterName=CrevasseMinLevel})
  n84 = ScalarParameter() {defaultValue=1, group=Constellation, parameterName=CrevasseFresnelMaskPowWeapons}
  n85 = ScalarParameter() {defaultValue=1, group=Constellation, parameterName=CrevasseFresnelMaskPowVehicles}
  n86 = Multiply(
   A=VectorParameter() {group=Constellation, parameterName=CS_EmissiveColor, defaultValue=<rgba 0.006 0.9647 0.956 1>}.RGB
   B=Power(Base=@n83, Exp=LinearInterpolate(A=@n84, B=@n85, Alpha=@n25))
  )
  n87 = Power(
   Base=Subtract(A=@n3, B=ScalarParameter() {group=Constellation, parameterName=CS_EmissiveFalloff2}) {constB=0.35}
   Exp=ScalarParameter() {defaultValue=2, group=Constellation, parameterName=CS_EmissvePower2}
  )
  n88 = ScalarParameter() {defaultValue=0.8, group=Constellation, parameterName=CrevasseFresnelMaskPow2Weapons}
  n89 = ScalarParameter() {defaultValue=0.8, group=Constellation, parameterName=CrevasseFresnelMaskPow2Vehicles}
  n90 = Multiply(
   A=Multiply(A=Clamp(None=@n87), B=VectorParameter() {group=Constellation, parameterName=CS_EmissiveColor2, defaultValue=<rgba 0 0.1667 0.5573 1>}.RGB)
   B=Power(Base=@n83, Exp=LinearInterpolate(A=@n88, B=@n89, Alpha=@n25))
  )
  n91 = SetMaterialAttributes!(
   MaterialAttributes=@n15.MaterialAttributes
   Specular=@n75
   Roughness=@n76
   Metallic=Constant()
   BaseColor=Add(A=Multiply(A=@n73, B=@n78), B=Add(A=@n86, B=@n90))
  ) {attributeSetTypes [
    #4
    #5
    #3
    #6
   ]}
  n92 = GetMaterialAttributes(None=@n91) {attributeGetTypes [#6, #1, #2]}
  n93 = VectorParameter() {group=Constellation, parameterName=CS_BaseColor, sortPriority=-101, defaultValue=<rgba 0.0123 0.0123 0.0137 1>}
  n94 = VectorParameter() {group=Constellation, parameterName=CS_BaseColor_Fresnel, sortPriority=-101, defaultValue=<rgba 0 0 0.006 0>}
  n95 = Fresnel(ExponentIn=ScalarParameter() {defaultValue=2, group=Constellation, parameterName=BaseColor_FresnelExponent, sortPriority=-101}) {exponent=2}
  n96 = Time()
  n97 = Desaturation(None=@n92.BaseColor)
  n98 = Constant3Vector() {constant=<rgba 0.1015 0 1 0>}
  n99 = Constant3Vector() {constant=<rgba 0.855 0.0343 0.7913 0>}
  n100 = Constant3Vector() {constant=<rgba 0.1022 0 1 1>}
  n101 = ComponentMask(None=TextureCoordinate() {coordinateIndex=1}) {g=true}
  n102 = Frac(None=Panner(Coordinate=Divide(A=@n101) {constB=140}) {speedX=0.1})
  n103 = MaterialFunctionCall(
   Hue Shift Percentage=Frac(None=Multiply(A=@n96) {constB=0.2})
   Texture=Multiply(
    A=@n97
    B=MaterialFunctionCall(A=@n98, B=@n99, C=@n100, Alpha=@n102) {materialFunction=&3ColorBlend}.Result
   )
  ) {materialFunction=&HueShift}
  n104 = ScalarParameter() {defaultValue=1, group=Constellation_Hueshift, parameterName=CS_BaseColorMultiplier}
  n105 = ComponentMask(None=@n73) {r=true}
  n106 = Fresnel() {baseReflectFraction=0.1, exponent=2}
  n107 = ScalarParameter() {defaultValue=200, parameterName=depthMin}
  n108 = Subtract(A=Distance(A=WorldPosition().XYZ, B=CameraPositionWS()), B=@n107)
  n109 = Subtract(A=ScalarParameter() {defaultValue=601, parameterName=depthMax}, B=@n107)
  n110 = LinearInterpolate(
   A=ScalarParameter() {defaultValue=0.3, parameterName=widthMin}
   B=ScalarParameter() {defaultValue=1, parameterName=widthMax}
   Alpha=Saturate(None=Divide(A=@n108, B=@n109))
  )
  n111 = ScalarParameter() {defaultValue=0.5, parameterName="Flatten (S)"}
  n112 = SetMaterialAttributes!(
   MaterialAttributes=@n92.MaterialAttributes
   BaseColor=LinearInterpolate(A=LinearInterpolate(A=@n93.RGB, B=@n94.RGB, Alpha=@n95), B=Multiply(A=@n103.Result, B=@n104), Alpha=Multiply(A=@n105) {constB=5})
   EmissiveColor=Multiply(
    A=Multiply(A=@n103.Result, B=@n106)
    B=ScalarParameter() {defaultValue=1.5, group=Constellation_Hueshift, parameterName=CS_EmissiveMultiplier, sortPriority=10}
   )
   Roughness=@n76
   Specular=@n75
   Normal=LinearInterpolate(
    A=@n92.Normal
    B=MaterialFunctionCall!(widthTip=@n110, Flatten=@n111, widthBase=@n110) {materialFunction=/Game/Characters/Player/Male/Medium/Bodies/M_MED_Neon/Materials/MF_splineThickenM}.normal
    Alpha=Multiply(A=@n97) {constB=20}
   )
  ) {attributeSetTypes [
    #6
    #1
    #5
    #4
    #2
   ]}
  n113 = GetMaterialAttributes(None=@n112) {attributeGetTypes [#1]}
  n114 = ChannelMaskParameter(None=TextureSample(UVs=@n42, Tex=@n49) {constMipValue=0, texture=&T_Snowflake_Pattern}.RGBA) {group=Constellation, parameterName=CS_Layer0Transition_TextureChannel, sortPriority=173}
  n115 = ScalarParameter() {defaultValue=1, group=Constellation, parameterName=CS_HeightLerp_Gradient_Clamp Max, sortPriority=174}
  n116 = Saturate(None=Multiply(
    A=Clamp(None=@n114.R, Max=@n115) {maxDefault=0.99, minDefault=0.01}
    B=ScalarParameter() {defaultValue=1, group=Constellation, parameterName=CS_HeightLerp_Gradient_Opacity_CE, sortPriority=175}
   ))
  n117 = Multiply(A=@n37, B=ScalarParameter() {defaultValue=0.1, group=Constellation, parameterName=CS_HeightLerp_Tiling, sortPriority=170})
  n118 = VectorParameter() {group=Constellation, parameterName=CS_HeightLerp_PanSpeedAndDirection, sortPriority=171, defaultValue=<rgba 0.035 -0.25 0 1>}
  n119 = ComponentMask(None=@n118.RGB) {g=true, r=true}
  n120 = ChannelMaskParameter(None=TextureSample(UVs=Panner(Coordinate=@n117, Speed=@n119), Tex=@n49) {constMipValue=0, texture=&T_Snowflake_Pattern}.RGBA) {group=Constellation, parameterName=CS_HeightTexture_Gradient_TextureChannel, sortPriority=172}
  n121 = ScalarParameter() {defaultValue=1, group=Constellation, parameterName=CS_HeightLerp_Contrast, sortPriority=176}
  n122 = Multiply(
   A=Saturate(None=MaterialFunctionCall(Transition Phase=@n116, Height Texture=@n120.R, Contrast=@n121) {materialFunction=&HeightLerp}.Alpha)
   B=VectorParameter() {group=Constellation, parameterName=Layer0_color, defaultValue=<rgba 0.0314 3.8406 5 1>}.RGB
  )
  n123 = LinearInterpolate(
   A=Multiply(A=ScalarParameter() {defaultValue=4, group=Constellation, parameterName=CS_Reactive_GlowIntensity, sortPriority=185}, B=@n122) {constB=2}
   B=@n122
   Alpha=@n58
  )
  n124 = StaticSwitchParameter(
   True=SetMaterialAttributes!(MaterialAttributes=@n113.MaterialAttributes, EmissiveColor=Add(A=@n113.EmissiveColor, B=@n123)) {attributeSetTypes [#1]}
   False=@n14
  ) {desc=Screen Space Projection - ambient Hueshift and Height Lerp - Reactive Glow on Fire, group=Customization_Wrap, parameterName=Wrap_Stars}
  FunctionOutput(None=@n124)
  n125 = Multiply(A=Subtract(A=@n1.Max, B=@n1.Max)) {constB=0.5}
  LinearInterpolate(A=TransformPosition(None=Add(A=@n1.Min, B=@n125)) {transformType=TRANSFORMPOSSOURCE_World}, B=@n16, Alpha=@n24)
  n126 = VectorParameter() {parameterName=Axis}
  n127 = ScalarParameter() {defaultValue=1, group=Constellation, parameterName=ScreenCoordScale_Weapons}
  n128 = ScalarParameter() {defaultValue=1, group=Constellation, parameterName=ScreenCoordScale_Vehicles}
  MaterialFunctionCall(
   Plane Normal Axis=@n17
   Plane X Axis=Normalize(VectorInput=CrossProduct(A=@n17, B=@n126.RGB))
   Plane Center=@n16
   UV Scale=Multiply(A=@n20, B=LinearInterpolate(A=@n127, B=@n128, Alpha=@n23))
  ) {materialFunction=&VirtualPlaneCoordinatesAbs}
  n129 = Distance(A=ComponentMask(None=ActorPositionWS()) {b=true, g=true, r=true}, B=Constant3Vector())
  n130 = OneMinus(None=Round(None=Saturate(None=@n129)))
  n131 = LinearInterpolate(A=ObjectRadius(), B=Constant() {r=100}, Alpha=Saturate(None=Multiply(A=@n130, B=@n25) {constA=1}))
  n132 = ActorPositionWS()
  n133 = Max(
   A=ComponentMask(None=TransformPosition(None=@n132) {transformSourceType=TRANSFORMPOSSOURCE_World, transformType=TRANSFORMPOSSOURCE_View}) {b=true}
   B=Constant() {r=100}
  )
  MaterialFunctionCall(
   UVs=@n13.PS
   Texture Scale=Multiply(A=Divide(A=@n131, B=@n133), B=ScalarParameter() {defaultValue=1, group=Constellation, parameterName=CS_UVRadialScaleZero})
  ) {materialFunction=&ScaleUVsByCenter}
  n134 = ComponentMask(None=@n29) {b=true, r=true}
  Multiply(A=Multiply(A=Sign(None=@n134)) {constB=-1}, B=@n32)
  n135 = VectorParameter() {group=Constellation_Hueshift, parameterName=CS_MaskTransitionPosition_Offset, sortPriority=-10, defaultValue=<rgba 0 0 0 1>}
  n136 = ObjectRadius()
  n137 = ScalarParameter() {defaultValue=1, group=Constellation_Hueshift, parameterName=CS_MaskRadiusMultiplier, sortPriority=-9}
  n138 = VertexInterpolator(VS=PreSkinnedPosition())
  Multiply(
   A=SphereMask(
    A=FunctionInput(Preview=@n138.PS) {bUsePreviewValueAsDefault=true, inputName=Local Position}
    B=@n135.RGB
    Radius=Multiply(A=@n136, B=@n137)
    Hardness=ScalarParameter() {group=Constellation_Hueshift, parameterName=CS_MaskGradientHardness, sortPriority=-8}
   )
   B=@n102
  )
  NamedRerouteDeclaration(None=FunctionInput() {bUsePreviewValueAsDefault=true, inputName=isCharacter, inputType=FunctionInput_Scalar}) {name=IsCharacter, nodeColor=<rgba 0.3125 0 1 1>}
  n139 = FunctionInput() {bUsePreviewValueAsDefault=true, inputName=IsStatsB, inputType=FunctionInput_Vector2, sortPriority=78}
  NamedRerouteDeclaration(None=ComponentMask(None=NamedRerouteDeclaration(None=@n139) {name=Is_StatsB, nodeColor=<rgba 0 0.9253 1 1>}) {r=true}) {name=IsBackpack, nodeColor=<rgba 0.2656 0 1 1>}
  ObjectPositionWS()
  ObjectRadius()
  Transform(None=CameraVectorWS()) {transformSourceType=TRANSFORMSOURCE_World, transformType=TRANSFORM_Local}
 ]}
