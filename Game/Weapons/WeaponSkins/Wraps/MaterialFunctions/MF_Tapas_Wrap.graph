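# Tapas wrap material function for weapons, vehicles and mechs, replacing surface attributes with black base color and animated emissive behind the Wrap_Tapas static switch.
# Camera-vector triplanar stars are layered with a bounds-projected detail mask and crevasse-masked Fresnel glows.
# A decay timed from Blueprint-supplied game time drives the reactive UV warp, noise and curve-atlas flicker.
# An older virtual-plane star layer remains in the graph, unconnected to the output.
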MaterialFunction /Game/Weapons/WeaponSkins/Wraps/MaterialFunctions/MF_Tapas_Wrap
graph {nodes [
  Constant3Vector() {constant=<rgba 0 0 128 1>}
  Constant3Vector() {constant=<rgba 0 0 80 1>}
  Constant() {r=40}
  Constant() {r=0.15}
  Constant() {r=70}
  Constant() {r=-0.005}
  Constant() {r=0.005}
  Constant() {r=75}
  Constant() {r=150}
  Constant() {r=0.5}
  n1 = VectorParameter() {group=Tapas_ScreenSpace, parameterName=TA_EmissiveColor, defaultValue=<rgba 0.006 0.9647 0.956 1>}
  n2 = FunctionInput() {inputName="Customization Crevasse (_CM G)", inputType=FunctionInput_Scalar, sortPriority=21, previewValue=<1 0 0 1>}
  n3 = Constant()
  n4 = ScalarParameter() {defaultValue=0.5, group=Tapas_ScreenSpace, parameterName=TA_NoCrevasseFresnelMaskValue}
  n5 = Abs(None=Add(A=If(A=@n2, B=@n3, "A > B"=@n2, "A == B"=@n4, "A < B"=@n2)) {constB=-0.5})
  n6 = Max(
   A=Saturate(None=Multiply(A=@n5) {constB=2})
   B=ScalarParameter() {defaultValue=0.12, group=Tapas_ScreenSpace, parameterName=TA_CrevasseMinLevel}
  )
  n7 = FunctionInput() {inputName=IsVehicle, inputType=FunctionInput_Scalar, sortPriority=4}
  n8 = FunctionInput() {inputName=IsMech, inputType=FunctionInput_Scalar, sortPriority=4}
  n9 = ComponentMask(None=Saturate(None=Add(A=@n7, B=@n8))) {r=true}
  n10 = Power(
   Base=@n6
   Exp=LinearInterpolate(
    A=ScalarParameter() {defaultValue=1, group=Tapas_ScreenSpace, parameterName=TA_CrevasseFresnelMaskPowWeapons}
    B=ScalarParameter() {defaultValue=1, group=Tapas_ScreenSpace, parameterName=TA_CrevasseFresnelMaskPowVehicles}
    Alpha=@n9
   )
  )
  n11 = PixelNormalWS()
  n12 = VertexNormalWS()
  n13 = OneMinus(None=DotProduct(A=CameraVectorWS(), B=StaticSwitchParameter(True=@n11, False=@n12) {group=Tapas_ScreenSpace, parameterName=TA_FresnelUsesPixelNormals}))
  n14 = Subtract(A=@n13, B=ScalarParameter() {group=Tapas_ScreenSpace, parameterName=TA_EmissiveFalloff2}) {constB=0.35}
  n15 = ScalarParameter() {defaultValue=2, group=Tapas_ScreenSpace, parameterName=TA_EmissvePower2}
  n16 = Multiply(
   A=Clamp(None=Power(Base=@n14, Exp=@n15))
   B=VectorParameter() {group=Tapas_ScreenSpace, parameterName=TA_EmissiveColor2, defaultValue=<rgba 0 0.1667 0.5573 1>}.RGB
  )
  n17 = Power(
   Base=@n6
   Exp=LinearInterpolate(
    A=ScalarParameter() {defaultValue=0.8, group=Tapas_ScreenSpace, parameterName=TA_CrevasseFresnelMaskPow2Weapons}
    B=ScalarParameter() {defaultValue=0.8, group=Tapas_ScreenSpace, parameterName=TA_CrevasseFresnelMaskPow2Vehicles}
    Alpha=@n9
   )
  )
  n18 = ScalarParameter() {group=Tapas_TopSide, parameterName=TA_StarCrevice_EmissiveStrength}
  n19 = TextureObjectParameter() {
   group=Tapas_ScreenSpace
   parameterName=TA_StarsTexture
   texture=/Game/Characters/Player/Female/Medium/Bodies/F_MED_Rebirth_Soldier/Skins/Constellation/Textures/T_Constellation_Pattern
  }
  n20 = FunctionInput() {bUsePreviewValueAsDefault=true, inputName=Warp Mask, inputType=FunctionInput_Scalar}
  n21 = ScalarParameter() {defaultValue=0.012, group=Tapas, parameterName=TA_WarpIntensity}
  n22 = Time()
  n23 = LinearInterpolate(A=@n22, B=Multiply(A=@n22, B=Constant()), Alpha=ScalarParameter() {group=Tapas, parameterName=TA_Background Rotation})
  n24 = AppendVector(
   A=AppendVector(
    A=ScalarParameter() {defaultValue=0.02, group=Tapas, parameterName=TA_Galaxy Rotation Speed}
    B=ScalarParameter() {defaultValue=-0.02, group=Tapas, parameterName=TA_Galaxy Rotation Speed_1}
   )
   B=ScalarParameter() {defaultValue=0.1, group=Tapas, parameterName=TA_Galaxy Rotation Speed_2}
  )
  n25 = ComponentMask(None=MaterialFunctionCall!(
    CameraVector
    Rotation Speed=Multiply(A=@n23, B=@n24)
    Texture=@n19
    Tiling=ScalarParameter() {defaultValue=0.5, group=Tapas, parameterName=TA_GalaxyTiling}
    Offset=Multiply(A=@n20, B=@n21)
    AxisFadeContrast=ScalarParameter() {defaultValue=6, group=Tapas, parameterName=TA_GalaxyAxisFade}
   ) {materialFunction=/Game/Weapons/WeaponSkins/MaterialFunction/TriplanarCameraVector_Rotated}.Result) {g=true}
  n26 = Multiply(A=@n25, B=VectorParameter() {group=Tapas, parameterName=TA_Star Brightness, defaultValue=<rgba 300 300 300 1>}.RGB)
  n27 = ComponentMask(None=MaterialFunctionCall!(
    CameraVector
    Texture=@n19
    Tiling=ScalarParameter() {defaultValue=2, group=Tapas, parameterName=TA_Bright Star Tiling}
    Offset
    AxisFadeContrast
   ) {materialFunction=/Engine/Functions/Engine_MaterialFunctions01/Texturing/TriplanarCameraVector}.Result) {r=true}
  n28 = TextureSampleParameter2D(UVs=Panner(Coordinate=TextureCoordinate() {uTiling=2, vTiling=2}) {speedX=0.133, speedY=0.2}) {group=Tapas_TopSide, parameterName=TA_NoiseTexture, sortPriority=99, texture=&T_BlacksRGB}
  n29 = Multiply(A=Multiply(A=@n27, B=@n27), B=ChannelMaskParameter(None=@n28.RGBA) {group=Tapas, parameterName=TA_StarNoise_ChannelMask, sortPriority=99}.R)
  n30 = Constant() {r=1}
  n31 = Constant() {r=3}
  n32 = Saturate(None=Multiply(
    A=MaterialFunctionCall(Brightness=@n30, Exponent=@n31) {materialFunction=&MF_Fresnel_FX}.Fresnel
    B=Constant() {r=2}
   ))
  n33 = Multiply(
   A=Multiply(
    A=LinearInterpolate(A=@n25, B=@n26, Alpha=@n29)
    B=VectorParameter() {group=Tapas, parameterName=TA_StarColor, defaultValue=<rgba 0 9.3228 20 1>}.RGB
   )
   B=@n32
  )
  n34 = ScalarParameter() {group=Tapas_TopSide, parameterName=TA_NoiseScrollingSpeedU, sortPriority=100}
  n35 = ScalarParameter() {group=Tapas_TopSide, parameterName=TA_NoiseScrollingSpeedV, sortPriority=100}
  n36 = Add(
   A=TextureCoordinate()
   B=Multiply(A=Time(), B=MaterialFunctionCall(X=@n34, Y=@n35) {materialFunction=&MakeFloat2}.Result)
  )
  n37 = ScalarParameter() {group=Tapas_TopSide, parameterName=TA_NoiseRotation, sortPriority=101}
  n38 = TextureSampleParameter2D(UVs=Multiply(
    A=MaterialFunctionCall(UVs=@n36, "Rotation Angle (0-1)"=@n37) {materialFunction=&CustomRotator}.Rotated Values
    B=ScalarParameter() {defaultValue=1, group=Tapas_TopSide, parameterName=TA_NoiseTiling, sortPriority=102}
   )) {group=Tapas_TopSide, parameterName=TA_NoiseTexture, sortPriority=99, texture=&T_BlacksRGB}
  n39 = ChannelMaskParameter(None=@n38.RGBA) {group=Tapas_Reactive, parameterName=TA_Reactive_Noise_ChannelMask, sortPriority=99}
  n40 = VectorParameter() {group=Tapas_TopSide, parameterName=TA_SurfaceDetailsColorA_NoiseColor, sortPriority=100, defaultValue=<rgba 0.0395 0.0319 0.1912 1>}
  n41 = Multiply(
   A=Power(Base=Divide(A=@n39.R) {constB=3}, Exp=ScalarParameter() {defaultValue=1, group=Tapas_TopSide, parameterName=TA_NoisePow, sortPriority=98})
   B=@n40.RGB
  )
  n42 = ScalarParameter() {group=Tapas_Reactive, parameterName=TA_Reactive_NoiseScrollingSpeedU, sortPriority=100}
  n43 = ScalarParameter() {group=Tapas_Reactive, parameterName=TA_Reactive_NoiseScrollingSpeedV, sortPriority=100}
  n44 = Add(
   A=TextureCoordinate()
   B=Multiply(A=Time(), B=MaterialFunctionCall(X=@n42, Y=@n43) {materialFunction=&MakeFloat2}.Result)
  )
  n45 = ScalarParameter() {group=Tapas_Reactive, parameterName=TA_Reactive_NoiseRotation, sortPriority=101}
  n46 = TextureSampleParameter2D(UVs=Multiply(
    A=MaterialFunctionCall(UVs=@n44, "Rotation Angle (0-1)"=@n45) {materialFunction=&CustomRotator}.Rotated Values
    B=ScalarParameter() {defaultValue=1, group=Tapas_Reactive, parameterName=TA_Reactive_NoiseTiling, sortPriority=102}
   )) {group=Tapas_Reactive, parameterName=TA-WarpUVS-01-Texture, sortPriority=99, texture=&T_BlacksRGB}
  n47 = ChannelMaskParameter(None=@n46.RGBA) {group=Tapas_Reactive, parameterName=TA_Reactive_EmissiveVariation_ChannelMask, sortPriority=99}
  n48 = VectorParameter() {
   group=Tapas_Reactive
   parameterName=TA_Reactive_SurfaceDetailsColorB_NoiseColor
   sortPriority=100
   defaultValue=<rgba 0.0678 0.1907 0.4735 1>
  }
  n49 = Multiply(A=@n47.R, B=@n48.RGB)
  n50 = Time()
  n51 = ScalarParameter() {group=00-Reactive-GTFB, parameterName=00-GameTimeFromBP, sortPriority=297}
  n52 = Saturate(None=Divide(
    A=Subtract(A=@n50, B=@n51)
    B=ScalarParameter() {defaultValue=2.25, group=Tapas_Reactive, parameterName=TA_AnimationDuration, sortPriority=298}
   ))
  n53 = ScalarParameter() {defaultValue=0.25, group=Tapas_Reactive, parameterName=TA_Curve_Amount, sortPriority=299}
  n54 = ScalarParameter() {group=Tapas_Reactive, parameterName=TA_TestReactive, sortPriority=300}
  n55 = LinearInterpolate(A=OneMinus(None=Power(Base=@n52, Exp=@n53)) {desc=inverts curve from 0-1 to 1-0}, B=@n54, Alpha=@n54)
  n56 = VectorParameter() {group=Tapas_TopSide, parameterName="TA_Scale_Weapon_01 A=uniform", sortPriority=30, defaultValue=<rgba 1 1 1 1>}
  n57 = Multiply(A=@n56.RGB, B=@n56.RGB)
  n58 = VectorParameter() {group=Tapas_TopSide, parameterName="TA_Scale_Vehicle_01 A=uniform", sortPriority=31, defaultValue=<rgba 1 1 1 1>}
  n59 = VectorParameter() {group=Tapas_TopSide, parameterName="TA_Scale_Vehicle_Side_01 A=uniform", sortPriority=31, defaultValue=<rgba 1 1 1 1>}
  n60 = FunctionInput() {bUsePreviewValueAsDefault=true, inputName=Side Project Vehicle, inputType=FunctionInput_Scalar}
  n61 = LinearInterpolate(A=Multiply(A=@n58.RGB, B=@n58.RGB), B=Multiply(A=@n59.RGB, B=@n59.RGB), Alpha=@n60)
  n62 = VectorParameter() {group=Tapas_TopSide, parameterName="TA_Scale_Mech_01 A=uniform", defaultValue=<rgba 1 1 1 1>}
  n63 = Constant3Vector() {constant=<rgba 1 1 -1 0>}
  n64 = VertexInterpolator(VS=PreSkinnedPosition())
  n65 = PreSkinnedLocalBounds()
  n66 = FunctionInput(Preview=Divide(A=Subtract(A=@n64.PS, B=@n65.Min), B=Subtract(A=@n65.Max, B=@n65.Max))) {bUsePreviewValueAsDefault=true, inputName=Wrap Bounds, sortPriority=12}
  n67 = VectorParameter() {group=Tapas_TopSide, parameterName=TA_TextureOffset_Weapon, sortPriority=15, defaultValue=<rgba 0 0 0 1>}
  n68 = VectorParameter() {group=Tapas_TopSide, parameterName=TA_TextureOffset_Vehicle, sortPriority=15, defaultValue=<rgba 0 0 0 1>}
  n69 = VectorParameter() {group=Tapas_TopSide, parameterName=TA_TextureOffset_Vehicle_Side, sortPriority=15, defaultValue=<rgba 0 0 0 1>}
  n70 = LinearInterpolate(A=@n67.RGB, B=LinearInterpolate(A=@n68.RGB, B=@n69.RGB, Alpha=@n60), Alpha=@n7)
  n71 = VectorParameter() {group=Tapas_TopSide, parameterName=TA_TextureOffset_Mech, sortPriority=15, defaultValue=<rgba 0 0 0 1>}
  n72 = Multiply(
   A=Add(A=Multiply(A=@n66, B=@n63), B=LinearInterpolate(A=@n70, B=@n71.RGB, Alpha=@n8))
   B=LinearInterpolate(A=LinearInterpolate(A=@n57, B=@n61, Alpha=@n7), B=Multiply(A=@n62.RGB, B=@n62.RGB), Alpha=@n8)
  )
  n73 = TextureCoordinate()
  n74 = LinearInterpolate(
   A=ScalarParameter() {defaultValue=1, group=Tapas_Reactive, parameterName=TA-WarpUVS-01-TileRate Weapon, sortPriority=53}
   B=ScalarParameter() {defaultValue=1, group=Tapas_Reactive, parameterName=TA-WarpUVS-01-TileRate Vehicle, sortPriority=54}
   Alpha=@n7
  )
  n75 = Time() {period=1}
  n76 = VectorParameter() {group=Tapas_Reactive, parameterName=TA-WarpUVS-01-Anitmated Weapon, sortPriority=57, defaultValue=<rgba 0 0 0 1>}
  n77 = VectorParameter() {group=Tapas_Reactive, parameterName=TA-WarpUVS-01-Anitmated Vehicle, sortPriority=57, defaultValue=<rgba 0 0 0 1>}
  n78 = ComponentMask(None=LinearInterpolate(A=@n76.RGB, B=@n77.RGB, Alpha=@n7)) {g=true, r=true}
  n79 = TextureSampleParameter2D(UVs=Add(A=Multiply(A=@n73, B=@n74), B=Multiply(A=@n75, B=@n78))) {group=Tapas_Reactive, parameterName=TA-WarpUVS-01-Texture, sortPriority=52, texture=&T_BlacksRGB}
  n80 = Multiply(
   A=ChannelMaskParameter(None=@n79.RGB) {group=Tapas_Reactive, parameterName=TA-WarpUVS-01_Channel, sortPriority=51}.R
   B=LinearInterpolate(
    A=ScalarParameter() {group=Tapas_Reactive, parameterName=TA-WarpUVS-01-Amount Weapon, sortPriority=55}
    B=ScalarParameter() {group=Tapas_Reactive, parameterName=TA-WarpUVS-01-Amount_Vehicle, sortPriority=56}
    Alpha=@n7
   )
  )
  n81 = MaterialFunctionCall(Color=LinearInterpolate(A=@n72, B=Add(A=@n80, B=@n72), Alpha=@n55)) {materialFunction=&SplitComponents}
  n82 = AppendVector(A=@n81.G, B=@n81.G)
  n83 = LinearInterpolate(A=AppendVector(A=@n81.R, B=@n81.R), B=AppendVector(A=@n81.R, B=@n81.R), Alpha=@n60)
  n84 = OneMinus(None=@n81.B)
  n85 = TextureSampleParameter2D(UVs=LinearInterpolate(A=LinearInterpolate(A=@n82, B=@n83, Alpha=@n7), B=AppendVector(A=@n84, B=@n81.R), Alpha=@n8)) {group=Tapas_TopSide, parameterName=TA_SurfaceDetailsMask, texture=&T_BlacksRGB}
  n86 = LinearInterpolate(
   A=LinearInterpolate(A=Multiply(A=@n33, B=@n18), B=LinearInterpolate(A=@n41, B=@n49, Alpha=@n55), Alpha=@n85.A)
   B=VectorParameter() {group=Tapas_TopSide, parameterName=TA_SurfaceDetailsColorG, defaultValue=<rgba 0.974 0.224 0.0287 0>}.RGB
   Alpha=@n85.G
  )
  n87 = VectorParameter() {group=Tapas_TopSide, parameterName=TA_SurfaceDetailsColorR, defaultValue=<rgba 0.974 0.224 0.0287 0>}
  n88 = Power(Base=@n85.R, Exp=ScalarParameter() {defaultValue=1, group=Tapas_TopSide, parameterName=TA_SurfaceDetailsColorR_MaskPow, sortPriority=79})
  n89 = Add(
   A=Add(A=Multiply(A=@n1.RGB, B=@n10), B=Multiply(A=@n16, B=@n17))
   B=LinearInterpolate(
    A=LinearInterpolate(A=@n86, B=@n87.RGB, Alpha=@n88)
    B=VectorParameter() {group=Tapas_TopSide, parameterName=TA_SurfaceDetailsColorB, defaultValue=<rgba 0.974 0.224 0.0287 0>}.RGB
    Alpha=@n85.B
   )
  )
  n90 = TextureCoordinate()
  n91 = ScalarParameter() {defaultValue=1, group=Tapas_TopSide, parameterName=TA_EmissiveVariation_TileRate, sortPriority=201}
  n92 = Time() {period=1}
  n93 = VectorParameter() {group=Tapas_TopSide, parameterName=TA_EmissiveVariation_Animation, sortPriority=201, defaultValue=<rgba 0 0 0 1>}
  n94 = ComponentMask(None=@n93.RGB) {g=true, r=true}
  n95 = TextureSampleParameter2D(UVs=Add(A=Multiply(A=@n90, B=@n91), B=Multiply(A=@n92, B=@n94))) {group=Tapas_TopSide, parameterName=TA-WarpUVS-01-Texture, sortPriority=52, texture=&T_BlacksRGB}
  n96 = VectorParameter() {group=Tapas_TopSide, parameterName=TA_EmissiveVariationColor, sortPriority=200, defaultValue=<rgba 0.974 0.224 0.0287 0>}
  n97 = Power(
   Base=Multiply(A=ChannelMaskParameter(None=@n95.RGBA) {group=Tapas_TopSide, parameterName=TA-EmissiveVariation_ChannelMask, sortPriority=200}.R, B=@n96.RGB)
   Exp=ScalarParameter() {defaultValue=1, group=Tapas_TopSide, parameterName=TA_EmissiveVariation_MaskPow, sortPriority=202}
  )
  n98 = LinearInterpolate(
   A=@n89
   B=LinearInterpolate(
    A=@n89
    B=Add(A=@n97, B=@n89)
    Alpha=ScalarParameter() {desc="1 = on, 0 = off", group=Tapas_TopSide, parameterName=TA_UseEmissiveVariation, sortPriority=199}
   )
   Alpha=Add(A=@n85.R, B=@n85.R) {desc=Emissive Variation Mask}
  )
  n99 = ChannelMaskParameter(None=@n38.RGBA) {group=Tapas_TopSide, parameterName=TA_Reactive_Noise_ChannelMask, sortPriority=99}
  n100 = Multiply(
   A=Power(
    Base=Divide(A=@n99.R) {constB=3}
    Exp=ScalarParameter() {defaultValue=1, group=Tapas_Reactive, parameterName=TA_Reactive_NoisePow_01, sortPriority=98}
   )
   B=@n40.RGB
  )
  n101 = Multiply(
   A=Power(Base=@n47.R, Exp=ScalarParameter() {defaultValue=1, group=Tapas_Reactive, parameterName=TA_Reactive_NoisePow_02, sortPriority=106})
   B=@n48.RGB
  )
  n102 = Add(A=Time(), B=ScalarParameter() {defaultValue=0.25, group=Tapas_Reactive, parameterName=TA_RandomOffset, sortPriority=50})
  n103 = ScalarParameter() {defaultValue=0.2, group=Tapas_Reactive, parameterName=TA_RandomSpeed, sortPriority=50}
  n104 = CurveAtlasRowParameter(CurveTime=Frac(None=Multiply(A=@n102, B=@n103))) {
   atlas=&AthenaCurveAtlas01
   curve=/Game/Athena/SafeZone/Textures/StormLighting
   defaultValue=18
   group=Tapas_Reactive
   parameterName=TA_Curve1_1
   sortPriority=49
  }
  n105 = Add(A=@n98, B=Multiply(A=Multiply(A=@n100, B=@n101), B=Multiply(A=@n104.R, B=@n104.R)))
  n106 = ScalarParameter() {defaultValue=1, group=Tapas_Reactive, parameterName=TA_GlobalEmissiveMultiplier, sortPriority=200}
  n107 = FunctionInput() {inputName=Material, inputType=FunctionInput_MaterialAttributes, sortPriority=-200}
  n108 = GetMaterialAttributes(None=@n107) {attributeGetTypes [
    #1
    #2
    #3
    #4
    #5
   ]}
  n109 = SetMaterialAttributes!(
   MaterialAttributes=@n108.MaterialAttributes
   BaseColor=Constant()
   EmissiveColor=LinearInterpolate(A=@n98, B=Multiply(A=@n105, B=@n106), Alpha=@n55)
   Specular=ScalarParameter() {defaultValue=0.2, group=Tapas, parameterName=TA_Specular, sortPriority=-12}
   Roughness=ScalarParameter() {defaultValue=0.8, group=Tapas, parameterName=TA_Roughness, sortPriority=-13}
   Normal=Constant3Vector() {constant=<rgba 0 0 1 0>}
  ) {attributeSetTypes [
    #1
    #2
    #5
    #4
    #3
   ]}
  n110 = StaticSwitchParameter(True=@n109, False=@n107) {
   desc=Triplanar Camera Vector Projection Base - Topside Masking - Reactive - Height Lerp Emissive on Fire
   group=Customization_Wrap
   parameterName=Wrap_Tapas
  }
  FunctionOutput(None=@n110)
  n111 = ScalarParameter() {defaultValue=0.25, group=Tapas, parameterName=Tapas-StarNoiseTile}
  n112 = Constant() {desc=Needs at least a small amount of this or it will break, r=0.25}
  n113 = MaterialFunctionCall() {materialFunction=&CameraDirectionVector}
  n114 = Normalize(VectorInput=Multiply(A=Constant3Vector() {constant=<rgba 1 1 0 0>}, B=@n113.Result))
  n115 = Constant() {r=25}
  n116 = Add(
   A=ActorPositionWS()
   B=Multiply(A=Constant3Vector() {constant=<rgba 0 0 1 0>}, B=ScalarParameter() {defaultValue=40, group=Tapas, parameterName=Tapas-AcreenAlignOffset})
  )
  n117 = MaterialFunctionCall(
   Input=MaterialFunctionCall() {materialFunction=/Engine/Functions/Engine_MaterialFunctions02/Utility/FOV}."FOV "
   Input Low=Constant() {r=0.414}
   Input High=Constant() {r=1.091}
   Target Low=Constant()
   Target High=Constant() {r=1}
  ) {materialFunction=&RemapValueRange}
  n118 = MaterialFunctionCall(
   Plane Normal Axis=Add(A=Multiply(A=@n112, B=@n114), B=@n113.Result)
   Plane Center=Add(A=Multiply(A=@n113.Result, B=@n115), B=@n116)
   UV Scale=LinearInterpolate(A=Constant() {r=75}, B=Constant() {r=200}, Alpha=Saturate(None=@n117.Result))
  ) {materialFunction=&VirtualPlaneCoordinates}
  n119 = ScalarParameter() {defaultValue=0.003, group=Tapas, parameterName=Tapas-StarNoiseSpeed}
  n120 = Time()
  n121 = TextureSample(UVs=Add(A=Multiply(A=@n111, B=@n118.UVs), B=Multiply(A=@n119, B=@n120))) {samplerSource=SSM_Wrap_WorldGroupSettings, texture=/Game/Weapons/WeaponSkins/Wraps/Textures/Tapas/T_Wrap_Tapas_M2}
  n122 = Constant() {r=6}
  n123 = Constant() {r=-0.0025}
  n124 = TextureSample(UVs=MaterialFunctionCall(UVs=@n118.UVs, "Rotation Angle (0-1)"=Multiply(A=@n120, B=@n123)) {materialFunction=&CustomRotator}.Rotated Values) {samplerSource=SSM_Wrap_WorldGroupSettings, texture=&T_Wrap_Tapas_M2}
  Multiply(A=Multiply(A=Power(Base=@n121.R, Exp=@n122), B=@n124.A), B=Constant() {r=15})
  n125 = Time()
  n126 = ScalarParameter() {defaultValue=-0.125, group=Tapas, parameterName=TA_Small Star Rotation speed}
  n127 = ComponentMask(None=MaterialFunctionCall(
    Rotation Speed=Multiply(A=@n125, B=@n126)
    Texture=@n19
    Tiling=ScalarParameter() {defaultValue=5, group=Tapas, parameterName=TA_Small Star Tiling}
   ) {materialFunction=&TriplanarCameraVector_Rotated}.Result) {b=true}
  Multiply(A=@n127, B=@n127)
  n128 = Subtract(A=@n13, B=ScalarParameter() {group=Tapas_ScreenSpace, parameterName=TA_EmissiveFalloff}) {constB=0.35}
  n129 = ScalarParameter() {defaultValue=2, group=Tapas_ScreenSpace, parameterName=TA_EmissvePower}
  Multiply(A=Clamp(None=Power(Base=@n128, Exp=@n129)), B=@n1.RGB)
  n130 = OneMinus(None=@n121.R)
  n131 = Constant() {r=4}
  n132 = Multiply(A=@n118.UVs, B=ScalarParameter() {defaultValue=5, group=Tapas, parameterName=Tapas-StarBScale})
  n133 = Multiply(A=@n120, B=ScalarParameter() {defaultValue=0.0005, group=Tapas, parameterName=Tapas-StarBSpeed})
  n134 = Multiply(
   A=LinearInterpolate(A=Constant() {r=0.005}, Alpha=Power(Base=@n130, Exp=@n131))
   B=TextureSample(UVs=MaterialFunctionCall(UVs=@n132, "Rotation Angle (0-1)"=@n133) {materialFunction=&CustomRotator}.Rotated Values) {samplerSource=SSM_Wrap_WorldGroupSettings, texture=&T_Wrap_Tapas_M2}.A
  )
  n135 = Constant() {r=10}
  n136 = Multiply(
   A=Add(A=@n124.A, B=Multiply(A=@n134, B=@n135))
   B=VectorParameter() {group=Tapas, parameterName=Tapas-EmBrightStars, defaultValue=<rgba 1.8 3.4747 6 1>}.RGB
  )
  n137 = SphereMask(
   A=Constant2Vector() {g=0.5, r=0.5}
   B=@n118.UVs
   Radius=ScalarParameter() {defaultValue=0.4, group=Tapas, parameterName=Tapas-Blue-StarRad}
   Hardness=Constant()
  ) {attenuationRadius=0.5}
  n138 = ScalarParameter() {defaultValue=5, group=Tapas, parameterName=Tapas-Blue-StarPow}
  n139 = Multiply(
   A=Multiply(
    A=Power(Base=@n137, Exp=@n138)
    B=VectorParameter() {group=Tapas, parameterName=Tapas-Blue-StarColor, defaultValue=<rgba 0.045 0.1467 0.3 1>}.RGB
   )
   B=ScalarParameter() {group=Tapas, parameterName=Tapas-Blue}
  )
  Reroute(None=Multiply(A=@n32, B=Add(A=@n136, B=@n139)))
  ScalarParameter() {defaultValue=0.15, group=Tapas, parameterName=Tapas-Blue-StarAmt}
  StaticSwitchParameter(True=ObjectPositionWS(), False=@n116) {group=Tapas, parameterName=Tapas_UseObjectPos}
 ]}
