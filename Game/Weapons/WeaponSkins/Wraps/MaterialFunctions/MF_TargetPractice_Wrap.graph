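# Material function for the Target Practice wrap: colored diagonal bands and a glowing score notch, masked over incoming attributes behind a static switch.
# Pre-skinned position, adjusted per item type, is side-projected for quantized, curve-atlas-colored bands and the notch texture, which is hidden on vehicles and mechs.
# Time since TP_GameTimeFromBP drives reactive band-light blinking and notch glow.
# Kill-count digit UV math exists but never reaches the output.
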MaterialFunction /Game/Weapons/WeaponSkins/Wraps/MaterialFunctions/MF_TargetPractice_Wrap
graph {nodes [
  n1 = FunctionInput(Preview=VertexInterpolator(VS=PreSkinnedPosition()).PS) {bUsePreviewValueAsDefault=true, inputName=PreskionnedPosition}
  n2 = ComponentMask(None=@n1) {r=true}
  Clamp(None=Round(None=OneMinus(None=@n2)))
  ComponentMask() {g=true, r=true}
  n3 = FunctionInput(Preview=Constant()) {bUsePreviewValueAsDefault=true, inputName=isVehicle, inputType=FunctionInput_Scalar}
  n4 = FunctionInput(Preview=Constant()) {bUsePreviewValueAsDefault=true, inputName=isMech, inputType=FunctionInput_Scalar}
  n5 = LinearInterpolate(
   A=LinearInterpolate(
    A=VectorParameter() {group=TargetPractice, parameterName=TP_Weapon_NotchOffset, defaultValue=<rgba 0 -15 -12 1>}.RGB
    B=VectorParameter() {group=TargetPractice, parameterName=TP_Vehicle_NotchOffset, defaultValue=<rgba 0 -15 -12 1>}.RGB
    Alpha=@n3
   )
   B=VectorParameter() {group=TargetPractice, parameterName=TP_Mech_NotchOffset, defaultValue=<rgba 0 -15 -12 1>}.RGB
   Alpha=@n4
  )
  n6 = VectorParameter() {group=TargetPractice, parameterName=TP_Weapon_ScoreScale, defaultValue=<rgba 0.1 0.1 0.1 1>}
  n7 = VectorParameter() {group=TargetPractice, parameterName=TP_Vehicle_ScoreScale, defaultValue=<rgba 0.1 0.1 0.1 1>}
  n8 = Divide(
   A=Add(A=@n1, B=@n5)
   B=LinearInterpolate(
    A=LinearInterpolate(A=@n6.RGB, B=@n7.RGB, Alpha=@n3)
    B=VectorParameter() {group=TargetPractice, parameterName=TP_Mech_ScoreScale, defaultValue=<rgba 0.1 0.1 0.1 1>}.RGB
    Alpha=@n4
   )
  )
  n9 = MaterialFunctionCall(Float3=@n8) {materialFunction=&BreakOutFloat3Components}
  ComponentMask(None=MaterialFunctionCall(X=@n9.R, Y=OneMinus(None=@n9.G), Z=@n9.R) {materialFunction=&MakeFloat3}.Result) {b=true, g=true}
  n10 = GetMaterialAttributes(None=FunctionInput() {inputName=Material, inputType=FunctionInput_MaterialAttributes}) {attributeGetTypes [#1, #2]}
  n11 = FunctionInput() {bUsePreviewValueAsDefault=true, inputName=Vehicle Side, inputType=FunctionInput_Scalar}
  n12 = PreSkinnedLocalBounds()
  n13 = FunctionInput(Preview=@n12.Max) {bUsePreviewValueAsDefault=true, inputName=ObjectBounds_Max}
  n14 = FunctionInput(Preview=@n12.Min) {bUsePreviewValueAsDefault=true, inputName=ObjectBounds_Min}
  n15 = FunctionInput(Preview=Constant3Vector()) {bUsePreviewValueAsDefault=true, inputName=ObjectBounds_Center}
  n16 = VectorParameter() {group=TargetPractice, parameterName=TP_Weapon_BandsOffset, defaultValue=<rgba 0 0 0 1>}
  n17 = VectorParameter() {group=TargetPractice, parameterName=TP_Pickaxe_BandsOffset, defaultValue=<rgba 0 0 0 1>}
  n18 = NamedRerouteUsage()
  n19 = VectorParameter() {group=TargetPractice, parameterName=TP_Char_BandsOffset, defaultValue=<rgba 0 0 0 1>}
  n20 = VectorParameter() {group=TargetPractice, parameterName=TP_Backpack_BandOffset, defaultValue=<rgba 0 0 0 1>}
  n21 = NamedRerouteUsage()
  n22 = LinearInterpolate(
   A=LinearInterpolate(
    A=LinearInterpolate(A=@n16.RGB, B=@n17.RGB, Alpha=@n18)
    B=VectorParameter() {group=TargetPractice, parameterName=TP_Vehicle_BandsOffset, defaultValue=<rgba 0 0 0 1>}.RGB
    Alpha=@n3
   )
   B=LinearInterpolate(
    A=VectorParameter() {group=TargetPractice, parameterName=TP_Mech_BandsOffset, defaultValue=<rgba 0 0 0 1>}.RGB
    B=LinearInterpolate(A=@n19.RGB, B=@n20.RGB, Alpha=@n21)
    Alpha=NamedRerouteUsage()
   )
   Alpha=@n4
  )
  n23 = Divide(A=@n12.Max, B=VectorParameter() {group=TargetPractice, parameterName=TP_ScaleBssedBiasFactor, defaultValue=<rgba 256 256 256 1>}.RGB)
  n24 = VectorParameter() {group=TargetPractice, parameterName=TP_Weapon_BandsScale, defaultValue=<rgba 1 1 1 1>}
  n25 = VectorParameter() {group=TargetPractice, parameterName=TP_Pickaxe_BandsScale, defaultValue=<rgba 1 1 1 1>}
  n26 = NamedRerouteUsage()
  n27 = VectorParameter() {group=TargetPractice, parameterName=TP_Char_BandsScale, defaultValue=<rgba 1 1 1 1>}
  n28 = VectorParameter() {group=TargetPractice, parameterName=TP_Backpack_BandScale, defaultValue=<rgba 1 1 1 1>}
  n29 = NamedRerouteUsage()
  n30 = LinearInterpolate(
   A=LinearInterpolate(
    A=LinearInterpolate(A=@n24.RGB, B=@n25.RGB, Alpha=@n26)
    B=VectorParameter() {group=TargetPractice, parameterName=TP_Vehicle_BandsScale, defaultValue=<rgba 1 1 1 1>}.RGB
    Alpha=@n3
   )
   B=LinearInterpolate(
    A=VectorParameter() {group=TargetPractice, parameterName=TP_Mech_BandsScale, defaultValue=<rgba 1 1 1 1>}.RGB
    B=LinearInterpolate(A=@n27.RGB, B=@n28.RGB, Alpha=@n29)
    Alpha=NamedRerouteUsage()
   )
   Alpha=@n4
  )
  n31 = Divide(A=Add(A=@n1, B=Add(A=@n22, B=@n23)), B=@n30)
  n32 = FunctionInput() {bUsePreviewValueAsDefault=true, inputName=isCharacter, inputType=FunctionInput_Scalar}
  n33 = MaterialFunctionCall!(
   isVehicle=@n3
   isMech=@n4
   Side Project Vehicle=@n11
   In=MaterialFunctionCall!(
    is Vehicle=@n3
    Object Bounds Max Adjustment=@n13
    Object Bounds Min Adjustment=@n14
    Object Bounds Center Adjustment=@n15
    isMech=@n4
    Side Project Vehicle=@n11
    Local Position=@n31
   ) {materialFunction=&/MF_WrapBounds}.Result
   IsBackpack
   IsPickaxe
   isCharacter=@n32
  ) {materialFunction=&/MF_SideProjection_Math}
  n34 = ComponentMask(None=@n33.Result) {r=true}
  n35 = Add(A=@n33.Result, B=Constant2Vector() {g=0.075})
  n36 = ComponentMask(None=@n35) {g=true}
  n37 = Clamp(None=Round(None=@n36))
  n38 = ComponentMask(None=@n33.Result) {g=true}
  n39 = OneMinus(None=@n38)
  n40 = Add(
   A=LinearInterpolate(A=@n34, B=Add(A=@n34) {constB=0.03}, Alpha=@n37)
   B=Multiply(
    A=LinearInterpolate(A=@n38, B=@n39, Alpha=@n37)
    B=ScalarParameter() {defaultValue=-0.2, group=TargetPractice, parameterName=TP_DiagonalFactor}
   )
  )
  n41 = ScalarParameter() {defaultValue=8, group=TargetPractice, parameterName=TP_BandsResolution}
  n42 = Divide(A=Ceil(None=Multiply(A=@n40, B=@n41)), B=@n41)
  n43 = CurveAtlasRowParameter(CurveTime=@n42) {
   atlas=&AthenaCurveAtlas01
   curve=/Game/Weapons/WeaponSkins/Wraps/Textures/TargetPractice/T_TargetPractice_Colors
   defaultValue=208
   parameterName=ColorCurve
  }
  n44 = Multiply(A=@n42) {constB=2}
  n45 = Divide(
   A=Distance(A=Subtract(A=@n44), B=Constant())
   B=ScalarParameter() {defaultValue=0.51, group=TargetPractice, parameterName=TP_IsolateMaskRadius}
  )
  n46 = OneMinus(None=Clamp(None=Round(None=@n45)))
  n47 = Multiply(A=Multiply(A=@n43, B=@n46), B=ScalarParameter() {defaultValue=0.5, group=TargetPractice, parameterName=TP_BaseEmissive})
  n48 = ScalarParameter() {defaultValue=0.5, group=TargetPractice, parameterName=TP_LightsDefaultSpeed}
  n49 = ScalarParameter() {defaultValue=2, group=TargetPractice, parameterName=TP_LightsReactiveSpeed}
  n50 = Subtract(A=Time(), B=ScalarParameter() {defaultValue=-100000000, group=TargetPractice, parameterName=TP_GameTimeFromBP})
  n51 = Divide(A=@n50, B=ScalarParameter() {defaultValue=3, group=TargetPractice, parameterName=TP_ReactiveBlinkDuration})
  n52 = Sine(None=Add(A=Saturate(None=@n51)) {constB=-0.25})
  n53 = Ceil(None=Clamp(None=ConstantBiasScale(None=@n52)))
  n54 = Frac(None=Multiply(A=Time(), B=LinearInterpolate(A=@n48, B=@n49, Alpha=@n53)))
  n55 = Divide(B=Multiply(A=@n41) {constB=2})
  n56 = Constant() {r=1}
  n57 = Subtract(A=OneMinus(None=@n42), B=OneMinus(None=@n54))
  n58 = LinearInterpolate(A=Multiply(A=@n46, B=SphereMask(A=@n42, B=@n54, Radius=@n55, Hardness=@n56)), B=Multiply(A=Ceil(None=@n57), B=@n46), Alpha=@n53)
  n59 = ScalarParameter() {defaultValue=5, group=TargetPractice, parameterName=TP_BlinkingLight_Strength}
  n60 = Multiply(A=@n59, B=ScalarParameter() {defaultValue=2, group=TargetPractice, parameterName=TP_BlinkingLight_ReactiveMultiplier})
  n61 = VectorParameter() {group=TargetPractice, parameterName=TP_Tens_Color, defaultValue=<rgba 1 0.7836 0 1>}
  n62 = Multiply(A=MaterialFunctionCall(
    isVehicle=@n3
    isMech=@n4
    Side Project Vehicle=@n11
    In=MaterialFunctionCall(
     is Vehicle=@n3
     Object Bounds Max Adjustment=@n13
     Object Bounds Min Adjustment=@n14
     Object Bounds Center Adjustment=@n15
     isMech=@n4
     Local Position=@n8
    ) {materialFunction=&MF_WrapBounds}.Result
    isCharacter=@n32
   ) {materialFunction=&MF_SideProjection_Math}.Result)
  n63 = TextureSampleParameter2D(UVs=@n62) {group=TargetPractice, parameterName=TP_AtlasTexture, texture=/Game/Weapons/WeaponSkins/Wraps/Textures/TargetPractice/T_NotchShape}
  n64 = Constant2Vector() {g=0.5, r=0.5}
  n65 = Round(None=Abs(None=Subtract(A=@n62, B=@n64)))
  n66 = Constant() {r=1}
  n67 = OneMinus(None=Ceil(None=Desaturation(None=@n65, Fraction=@n66)))
  n68 = LinearInterpolate(A=Multiply(A=@n63.R, B=Clamp(None=@n67)), Alpha=Clamp(None=Add(A=@n3, B=@n4))) {constB=0}
  n69 = ScalarParameter() {defaultValue=2, group=TargetPractice, parameterName=TP_Score_BaseGlow}
  n70 = Multiply(A=@n69, B=ScalarParameter() {defaultValue=5, group=TargetPractice, parameterName=TP_ReactiveGlow_Multiplier})
  n71 = Saturate(None=Divide(A=@n50, B=ScalarParameter() {defaultValue=2, group=TargetPractice, parameterName=TP_ReactiveGlowDuration}))
  n72 = ConstantBiasScale(None=Sine(None=Add(A=@n71) {constB=-0.25}))
  n73 = Add(
   A=Multiply(A=Multiply(A=@n43, B=@n58), B=LinearInterpolate(A=@n59, B=@n60, Alpha=@n53))
   B=Multiply(A=Multiply(A=@n61.RGB, B=@n68), B=LinearInterpolate(A=@n69, B=@n70, Alpha=@n72))
  )
  n74 = Round(None=ComponentMask(None=@n62) {r=true})
  n75 = LinearInterpolate(A=@n61.RGB, B=VectorParameter() {group=TargetPractice, parameterName=TP_Units_Color, defaultValue=<rgba 1 0 0.507 1>}.RGB, Alpha=@n74)
  n76 = LinearInterpolate(
   A=VectorParameter() {group=TargetPractice, parameterName=TP_BG_Color, defaultValue=<rgba 0.0382 0.0222 0.0307 1>}.RGB
   B=Multiply(
    A=LinearInterpolate(A=@n43, B=@n75, Alpha=@n68)
    B=ScalarParameter() {defaultValue=0.5, group=TargetPractice, parameterName=TP_BaseColorMultiplier}
   )
   Alpha=Clamp(None=Add(A=@n46, B=@n68))
  )
  n77 = SetMaterialAttributes!(MaterialAttributes=@n10.MaterialAttributes, EmissiveColor=Add(A=Add(A=@n10.EmissiveColor, B=@n47), B=Power(Base=@n73)), BaseColor=@n76) {attributeSetTypes [#1, #2]}
  n78 = ChannelMaskParameter(None=FunctionInput(Preview=Constant3Vector()) {bUsePreviewValueAsDefault=true, inputName=Mask}) {group=TargetPractice, parameterName=TP_MaskChannel}
  FunctionOutput(None=StaticSwitchParameter(True=BlendMaterialAttributes(A=@n10.MaterialAttributes, B=@n77, Alpha=@n78.R), False=@n10.MaterialAttributes) {group=Customization_Wrap, parameterName=UseTargetPracticeWrap})
  LinearInterpolate(
   A=LinearInterpolate(A=ComponentMask(None=@n31) {b=true, g=true}, B=ComponentMask(None=@n31) {g=true, r=true}, Alpha=@n3)
   B=ComponentMask(None=@n31) {b=true, r=true}
   Alpha=@n4
  )
  LinearInterpolate(A=@n34, B=@n38, Alpha=@n4)
  LinearInterpolate(A=@n38, B=@n34, Alpha=@n4)
  LinearInterpolate(A=@n36, B=ComponentMask(None=@n35) {r=true}, Alpha=@n4)
  LinearInterpolate(A=@n10.BaseColor, B=@n76, Alpha=@n46)
  n79 = Multiply(A=@n62, B=Constant2Vector() {g=1.5, r=2}) {constB=2}
  n80 = MaterialFunctionCall(Float2=@n79) {materialFunction=&BreakOutFloat2Components}
  n81 = OneMinus(None=@n80.R)
  n82 = MaterialFunctionCall(Float2=@n79) {materialFunction=&BreakOutFloat2Components}
  n83 = Add(A=@n82.R) {constB=-1}
  n84 = OneMinus(None=@n82.G)
  n85 = MaterialFunctionCall(Float2=LinearInterpolate(
    A=MaterialFunctionCall(X=@n81, Y=@n80.G) {materialFunction=&MakeFloat2}.Result
    B=MaterialFunctionCall(X=@n83, Y=@n84) {materialFunction=&MakeFloat2}.Result
    Alpha=@n74
   )) {materialFunction=&BreakOutFloat2Components}
  n86 = Divide(A=Constant() {r=1}, B=ScalarParameter() {defaultValue=10, group=TargetPractice, parameterName=TP_No_Columns})
  n87 = ScalarParameter() {defaultValue=54, group=TargetPractice, parameterName=TP_KillsInput}
  n88 = Constant() {r=10}
  n89 = Divide(A=@n87, B=@n88) {constB=10}
  n90 = LinearInterpolate(A=Subtract(A=@n89, B=Frac(None=@n89)), B=Fmod(A=@n87, B=@n88), Alpha=@n74)
  MaterialFunctionCall(X=Add(A=Multiply(A=@n85.R, B=@n86), B=Multiply(A=@n86, B=@n90)) {constB=0}, Y=@n85.G) {materialFunction=&MakeFloat2}
  n91 = MaterialFunctionCall(Float2=@n62) {materialFunction=&BreakOutFloat2Components}
  MaterialFunctionCall(X=OneMinus(None=@n91.R), Y=@n91.G) {materialFunction=&MakeFloat2}
  Multiply(A=@n30, B=@n13)
  Multiply(A=@n5, B=Divide(A=@n13, B=Subtract(A=@n12.Max, B=@n12.Max)))
  Multiply(A=VectorParameter() {group=TargetPractice, parameterName=TP_Global_BandsScale, defaultValue=<rgba 1 1 1 1>}.RGB, B=@n30)
  Multiply(B=Constant2Vector() {g=1.5, r=1})
  NamedRerouteDeclaration(None=@n32) {name=IsCharacter, nodeColor=<rgba 0.3125 0 1 1>}
  n92 = NamedRerouteDeclaration(None=FunctionInput() {bUsePreviewValueAsDefault=true, inputName=IsStatsB, inputType=FunctionInput_Vector2, sortPriority=78}) {name=Is_StatsB, nodeColor=<rgba 0 0.9253 1 1>}
  NamedRerouteDeclaration(None=ComponentMask(None=@n92) {r=true}) {name=IsBackpack, nodeColor=<rgba 0.2656 0 1 1>}
  NamedRerouteDeclaration(None=ComponentMask(None=@n92) {g=true}) {name=IsPickaxe, nodeColor=<rgba 1 0 0.5625 1>}
  n93 = PreSkinnedNormal()
  Reroute(None=FunctionInput(Preview=VertexInterpolator(VS=@n93).PS) {bUsePreviewValueAsDefault=true, inputName=PreskinnedNormal})
  Round(None=ComponentMask(None=TextureCoordinate() {coordinateIndex=1}) {r=true})
  ScalarParameter() {defaultValue=0.5, parameterName=TP_DebugTimeSpeed}
 ]}
