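# TechOpsBlue wrap material function that turns weapons, vehicles, mechs or backpacks into glossy metal with animated, glowing blue waves.
# Camera-facing virtual-plane UVs, scaled by object bounds, drive three panning, noise-distorted wave-texture channels, with tiling and distortion selected by vehicle and mech flags.
# Waves fade along pre-skinned local Z over a two-color gradient with Fresnel rim.
# Fresnel masks blend this into BaseColor and Emissive, gated by the UseSonyWrap static switch.
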MaterialFunction /Game/Weapons/WeaponSkins/Wraps/MaterialFunctions/MF_TechOpsBlue
graph {nodes [
  n1 = FunctionInput() {inputName=MaterialInput, inputType=FunctionInput_MaterialAttributes}
  n2 = GetMaterialAttributes(None=@n1) {attributeGetTypes [#1, #2]}
  n3 = Multiply(A=@n2.BaseColor) {constB=2}
  n4 = Transform(None=@n2.Normal)
  n5 = Power(Base=Fresnel(ExponentIn=Constant() {r=1}, Normal=@n4))
  n6 = Normalize(VectorInput=Subtract(A=ObjectPositionWS(), B=CameraPositionWS()))
  n7 = VectorParameter() {group=TechOpsBlue, parameterName=Axis, defaultValue=<rgba 0 0 1 1>}
  n8 = MaterialFunctionCall() {materialFunction=/Engine/Functions/Engine_MaterialFunctions02/WorldPositionOffset/ObjectScale}
  n9 = PreSkinnedLocalBounds()
  n10 = MaterialFunctionCall(Float3=@n9.Extents) {materialFunction=&BreakOutFloat3Components}
  n11 = Max(A=Max(A=@n10.R, B=@n10.R), B=@n10.B)
  n12 = MaterialFunctionCall!(
   Plane Normal Axis=@n6
   Plane X Axis=Normalize(VectorInput=CrossProduct(A=@n6, B=@n7.RGB))
   Plane Center=ObjectPositionWS()
   UV Scale=Divide(A=Multiply(A=@n8.Scale X, B=@n11), B=ScalarParameter() {defaultValue=2, group=TechOpsBlue, parameterName=UVRadialScaleZero})
  ) {materialFunction=/Game/Athena/Environments/Nexus/Materials/VirtualPlaneCoordinatesAbs}
  n13 = Add(A=@n12.UVs) {constB=0.5}
  n14 = FunctionInput(Preview=ScalarParameter() {parameterName=IsVehicle}) {bUsePreviewValueAsDefault=true, inputName=IsVehicle, inputType=FunctionInput_Scalar}
  n15 = FunctionInput(Preview=ScalarParameter() {parameterName=IsMech}) {bUsePreviewValueAsDefault=true, inputName=IsMech, inputType=FunctionInput_Scalar}
  n16 = LinearInterpolate(
   A=LinearInterpolate(
    A=ScalarParameter() {defaultValue=0.5, group=TechOpsBlue, parameterName=WeaponWavesTiling}
    B=ScalarParameter() {defaultValue=0.5, group=TechOpsBlue, parameterName=VehicleWavesTiling}
    Alpha=@n14
   )
   B=ScalarParameter() {defaultValue=0.2, group=TechOpsBlue, parameterName=MechWavesTiling}
   Alpha=@n15
  )
  n17 = Multiply(A=@n12.UVs) {constB=0.25}
  n18 = Panner(Coordinate=@n17) {speedX=0.02, speedY=0.02}
  n19 = Multiply(A=ComponentMask(None=TextureSample(UVs=@n18) {samplerType=SAMPLERTYPE_LinearColor, texture=&HighContrastNoise}.RGB) {g=true, r=true}) {constB=2}
  n20 = LinearInterpolate(
   A=LinearInterpolate(
    A=ScalarParameter() {defaultValue=0.02, group=TechOpsBlue, parameterName=WeaponDistortionStrength}
    B=ScalarParameter() {defaultValue=0.02, group=TechOpsBlue, parameterName=VehicleDistortionStrength}
    Alpha=@n14
   )
   B=ScalarParameter() {defaultValue=0.02, group=TechOpsBlue, parameterName=MechDistortionStrength}
   Alpha=@n15
  )
  n21 = Add(A=Panner(Coordinate=Multiply(A=@n13, B=@n16)) {speedX=0.01}, B=Multiply(A=Subtract(A=@n19), B=@n20))
  n22 = Add(A=@n12.UVs) {constB=0.25}
  n23 = Panner(Coordinate=@n17) {speedX=0.01, speedY=0.03}
  n24 = Multiply(A=ComponentMask(None=TextureSample(UVs=@n23) {samplerType=SAMPLERTYPE_LinearColor, texture=&HighContrastNoise}.RGB) {g=true, r=true}) {constB=2}
  n25 = Add(A=Panner(Coordinate=Multiply(A=@n22, B=@n16) {constB=2}) {speedX=-0.015}, B=Multiply(A=Subtract(A=@n24), B=@n20))
  n26 = Add(
   A=Multiply(A=TextureSample(UVs=@n21) {texture=/Game/Global/Textures/Default/Tileable_Patterns/T_WavePattern_01}.G) {constB=1.5}
   B=Multiply(A=TextureSample(UVs=@n25) {texture=&T_WavePattern_01}.B) {constB=1.5}
  )
  n27 = Panner(Coordinate=Multiply(A=@n12.UVs, B=@n16) {constB=2}) {speedX=-0.02}
  n28 = Add(A=@n17) {constB=0.25}
  n29 = ComponentMask(None=TextureSample(UVs=Panner(Coordinate=@n28) {speedX=0.01, speedY=0.03}) {samplerType=SAMPLERTYPE_LinearColor, texture=&HighContrastNoise}.RGB) {g=true, r=true}
  n30 = Multiply(A=Subtract(A=Multiply(A=@n29) {constB=2}), B=@n20)
  n31 = Multiply(A=TextureSample(UVs=Add(A=@n27, B=@n30)) {texture=&T_WavePattern_01}.R) {constB=1.5}
  n32 = FunctionInput(Preview=VertexInterpolator(VS=PreSkinnedPosition()).PS) {bUsePreviewValueAsDefault=true, inputName=Pre-SkinnedLocalPosition}
  n33 = Subtract(A=@n32, B=FunctionInput(Preview=@n9.Min) {bUsePreviewValueAsDefault=true, inputName=Pre-SkinnedLocalBounds_Min})
  n34 = FunctionInput(Preview=@n9.Extents) {bUsePreviewValueAsDefault=true, inputName=Pre-SkinnedLocalBounds}
  n35 = Divide(A=Add(A=@n32, B=Abs(None=@n33)), B=@n34)
  n36 = Subtract(A=ComponentMask(None=@n35) {b=true}) {constB=0.25}
  n37 = Multiply(
   A=VectorParameter() {group=TechOpsBlue, parameterName=WaveColor, defaultValue=<rgba 0.0003 0.3916 0.7454 1>}.RGB
   B=Multiply(A=Add(A=@n26, B=@n31), B=Clamp(None=@n36))
  )
  n38 = ScalarParameter() {defaultValue=1.5, group=TechOpsBlue, parameterName=Wave_EmissivePower}
  n39 = VectorParameter() {group=TechOpsBlue, parameterName=BG_Color_02, defaultValue=<rgba 0.0091 0.2623 0.5841 1>}
  n40 = Subtract(A=@n35) {constB=0}
  n41 = Power(Base=ComponentMask(None=Clamp(None=@n40)) {b=true})
  n42 = Power(Base=Multiply(A=Multiply(A=@n39.RGB, B=@n41)))
  n43 = ComponentMask(None=NamedRerouteDeclaration(None=FunctionInput() {bUsePreviewValueAsDefault=true, inputName=IsStatsB, inputType=FunctionInput_Vector2, sortPriority=78}) {name=Is_StatsB, nodeColor=<rgba 0 0.9253 1 1>}) {r=true}
  n44 = Multiply(
   A=OneMinus(None=@n41)
   B=LinearInterpolate(
    A=VectorParameter() {group=TechOpsBlue, parameterName=BG_Color_01, defaultValue=<rgba 0.0015 0.0331 0.159 1>}.RGB
    B=VectorParameter() {group=TechOpsBlue, parameterName=BG_Color_Backpack_01, defaultValue=<rgba 1 1 1 1>}.RGB
    Alpha=NamedRerouteDeclaration(None=@n43) {name=IsBackpack, nodeColor=<rgba 0.2656 0 1 1>}
   )
  )
  n45 = Constant() {r=2}
  n46 = Clamp(None=Power(Base=Fresnel(ExponentIn=@n45, Normal=@n4)))
  n47 = Add(A=Power(Base=Multiply(A=@n37, B=@n38)), B=Add(A=Add(A=@n42, B=@n44), B=Multiply(A=@n39.RGB, B=@n46)) {constB=0})
  Add(A=Multiply(A=@n3, B=Clamp(None=@n5)), B=@n47)
  ComponentMask(None=@n34) {g=true}
  Constant2Vector() {g=1, r=1}
  n48 = Time()
  n49 = ScalarParameter() {defaultValue=0.1, group=TechOpsBlue, parameterName=Wave_01FadeSpeed}
  ConstantBiasScale(None=Sine(None=Multiply(A=@n48, B=@n49)))
  n50 = Multiply(A=@n48, B=@n49)
  ConstantBiasScale(None=Sine(None=Add(A=@n50) {constB=0.5}))
  n51 = ScalarParameter() {defaultValue=0.2, group=TechOpsBlue, parameterName=Wave_03FadeSpeed}
  ConstantBiasScale(None=Sine(None=Multiply(A=@n48, B=@n51)))
  Constant() {r=0.5}
  n52 = Constant() {r=0.5}
  n53 = Clamp(None=OneMinus(None=Fresnel(ExponentIn=@n52, Normal=@n4)))
  n54 = Constant() {r=2}
  n55 = Clamp(None=OneMinus(None=Fresnel(ExponentIn=@n54, Normal=@n4)))
  n56 = SetMaterialAttributes!(
   MaterialAttributes=@n2.MaterialAttributes
   BaseColor=Clamp(None=LinearInterpolate(A=@n3, B=@n47, Alpha=@n53))
   EmissiveColor=Clamp(None=Multiply(A=@n47, B=@n55))
   Metallic=Constant() {r=1}
   Roughness=Constant() {r=0.1}
   Specular=Constant() {r=1}
  ) {attributeSetTypes [
    #2
    #3
    #4
    #5
    #6
   ]}
  FunctionOutput(None=StaticSwitchParameter(True=@n56, False=@n1) {group=Customization_Wrap, parameterName=UseSonyWrap})
  VectorParameter() {group=TechOpsBlue, parameterName=FlatBaseCOlor, defaultValue=<rgba 1 1 1 1>}
 ]}
