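# Material function for the Tomato weapon-skin wrap, toggled by the Wrap_Tomato static switch, layering projected textures and reactive glow onto input material attributes.
# Top/side projection from pre-skinned bounds drives surface and emissive channels, with per-weapon, vehicle and mech transforms, blended through customization masks.
# Emissive supports panning and UV warp.
# A timer from a Blueprint-fed timestamp fades the reactive glow and front gradient, described as firing feedback.
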MaterialFunction /Game/Weapons/WeaponSkins/Wraps/MaterialFunctions/MF_Tomato_Wrap
graph {nodes [
  FunctionInput(Preview=Constant3Vector() {constant=<rgba 1 0 0 0>}) {bUsePreviewValueAsDefault=true, inputName=ForwardAxis, sortPriority=4}
  n1 = FunctionInput() {bUsePreviewValueAsDefault=true, inputName="Customization Mask (_CM B)", inputType=FunctionInput_Scalar, sortPriority=24}
  n2 = ScalarParameter() {group=TopSideProjection, parameterName=TS_Remove_Customization Mask, sortPriority=-10}
  n3 = FunctionInput() {bUsePreviewValueAsDefault=true, inputName="Customization Mask 2 (_CM B Top Right)", inputType=FunctionInput_Scalar, sortPriority=25}
  n4 = LinearInterpolate(
   A=Saturate(None=Add(A=@n1, B=@n2))
   B=Add(A=@n3, B=@n1)
   Alpha=ScalarParameter() {group=TopSideProjection, parameterName=TS_MasksForPrimary, sortPriority=-10}
  )
  n5 = Saturate(None=Subtract(A=NamedRerouteUsage(), B=NamedRerouteUsage()))
  n6 = FunctionInput() {inputName=Material, inputType=FunctionInput_MaterialAttributes, sortPriority=-10}
  n7 = GetMaterialAttributes(None=@n6) {attributeGetTypes [#1, #2, #3]}
  n8 = TextureObjectParameter() {
   group=TopSideProjection__Textures
   parameterName="TS_BaseColor_01 - RGB=BaseColor A=Roughness"
   sortPriority=11
   texture=&T_BlacksRGB
  }
  n9 = FunctionInput(Preview=Constant()) {bUsePreviewValueAsDefault=true, inputName=isMech, inputType=FunctionInput_Scalar, sortPriority=2}
  n10 = LinearInterpolate(
   A=ScalarParameter() {defaultValue=1, group=TopSideProjection, parameterName=TS_Blend_Contrast, sortPriority=-1}
   B=ScalarParameter() {defaultValue=2, group=TopSideProjection, parameterName=TS_Blend_Contrast_Mech, sortPriority=-1}
   Alpha=@n9
  )
  n11 = ScalarParameter() {group=TopSideProjection, parameterName=TS_Fill_alternate_Axis_Opacity 01, sortPriority=2}
  n12 = VectorParameter() {group=TopSideProjection, parameterName="TS_Fill_alternate_Axis_Color_A=Roughness 01", sortPriority=3}
  n13 = AppendVector(A=@n12.RGB, B=@n12.RGB)
  n14 = VectorParameter() {group=TopSideProjection_UVS, parameterName=TS-Rotation-Weapon-01---------------------, sortPriority=1, defaultValue=<rgba 0 0 -0.25 1>}
  n15 = VectorParameter() {group=TopSideProjection_UVS, parameterName=TS_Rotation_Vehicle_01, sortPriority=2, defaultValue=<rgba 0 0 0 1>}
  n16 = VectorParameter() {
   desc=must activate TS_Vehicle_Side_01 UVs
   group=TopSideProjection_UVS
   parameterName=TS_Rotation_Vehicle_Side_01
   sortPriority=2
   defaultValue=<rgba 0 0 0 1>
  }
  n17 = FunctionInput() {bUsePreviewValueAsDefault=true, inputName=Side Project Vehicle, inputType=FunctionInput_Scalar, sortPriority=3}
  n18 = Multiply(A=@n17, B=ScalarParameter() {group=TopSideProjection_UVS, parameterName="TS_Vehicle_Side_01 UVs 0=Off 1=Active"})
  n19 = FunctionInput(Preview=Constant()) {inputName=isVehicle, inputType=FunctionInput_Scalar, sortPriority=1}
  n20 = VectorParameter() {group=TopSideProjection_UVS, parameterName=TS_Rotation_Mech_01, sortPriority=3, defaultValue=<rgba 0.25 0.25 0 1>}
  n21 = LinearInterpolate(A=LinearInterpolate(A=@n14.RGB, B=LinearInterpolate(A=@n15.RGB, B=@n16.RGB, Alpha=@n18), Alpha=@n19), B=@n20.RGB, Alpha=@n9)
  n22 = FunctionInput(Preview=VertexInterpolator(VS=PreSkinnedNormal()).PS) {bUsePreviewValueAsDefault=true, inputName=Pre-Skinned Local Normal, sortPriority=10}
  n23 = VectorParameter() {group=TopSideProjection_UVS, parameterName=TS_Offset_Weapon_01---------------------, sortPriority=10, defaultValue=<rgba 0 0 0 1>}
  n24 = VectorParameter() {group=TopSideProjection_UVS, parameterName=TS_Offset_Vehicle_01, sortPriority=11, defaultValue=<rgba 0 0 0 1>}
  n25 = VectorParameter() {
   desc=must activate TS_Vehicle_Side_01 UVs
   group=TopSideProjection_UVS
   parameterName=TS_Offset_Vehicle_Side_01
   sortPriority=11
   defaultValue=<rgba 0 0 0 1>
  }
  n26 = LinearInterpolate(A=@n23.RGB, B=LinearInterpolate(A=@n24.RGB, B=@n25.RGB, Alpha=@n18), Alpha=@n19)
  n27 = VectorParameter() {group=TopSideProjection_UVS, parameterName=TS_Offset_Mech_01, sortPriority=12, defaultValue=<rgba 0 0 0 1>}
  n28 = PreSkinnedPosition()
  n29 = MaterialFunctionCall!(
   is Vehicle=@n19
   Object Bounds Max Adjustment=FunctionInput() {bUsePreviewValueAsDefault=true, inputName=Object Bounds Max Adjustment, sortPriority=8}
   Object Bounds Min Adjustment=FunctionInput() {bUsePreviewValueAsDefault=true, inputName=Object Bounds Min Adjustment, sortPriority=7}
   Object Bounds Center Adjustment=FunctionInput() {bUsePreviewValueAsDefault=true, inputName=Object Bounds Center Adjustment, sortPriority=9}
   isMech=@n9
   Side Project Vehicle
   Local Position=FunctionInput(Preview=VertexInterpolator(VS=@n28).PS) {bUsePreviewValueAsDefault=true, inputName=Local Position, sortPriority=6}
  ) {materialFunction=&/MF_WrapBounds}
  n30 = VectorParameter() {group=TopSideProjection_UVS, parameterName="TS_Scale_Weapon_01 A=uniform-----", sortPriority=30, defaultValue=<rgba 1 1 1 1>}
  n31 = Multiply(A=@n30.RGB, B=@n30.RGB)
  n32 = VectorParameter() {group=TopSideProjection_UVS, parameterName="TS_Scale_Vehicle_01 A=uniform", sortPriority=31, defaultValue=<rgba 1 1 1 1>}
  n33 = VectorParameter() {
   desc=must activate TS_Vehicle_Side_01 UVs
   group=TopSideProjection_UVS
   parameterName="TS_Scale_Vehicle_Side_01 A=uniform"
   sortPriority=31
   defaultValue=<rgba 1 1 1 1>
  }
  n34 = LinearInterpolate(A=Multiply(A=@n32.RGB, B=@n32.RGB), B=Multiply(A=@n33.RGB, B=@n33.RGB), Alpha=@n18)
  n35 = VectorParameter() {group=TopSideProjection_UVS, parameterName="TS_Scale_Mech_01 A=uniform", defaultValue=<rgba 1 1 1 1>}
  n36 = Multiply(
   A=Add(A=@n29.Result, B=LinearInterpolate(A=@n26, B=@n27.RGB, Alpha=@n9))
   B=LinearInterpolate(A=LinearInterpolate(A=@n31, B=@n34, Alpha=@n19), B=Multiply(A=@n35.RGB, B=@n35.RGB), Alpha=@n9)
  )
  n37 = MaterialFunctionCall!(
   IsCharacter
   IsStatsB
   Texture=@n8
   Rotation_Axis=@n21
   Caculated_Bounds=@n36
   Pre_Skinned_Normals=@n22
   Blend_Contrast=@n10
   Fill_alternate_Axis_Opacity=@n11
   "Fill_alternate_Axis_Color_A=Roughness"=@n13
   Is_Vehicle=@n19
   Is_Mech=@n9
   Enable_Fill_Z_Weapon
   Fill_Z_Weapon_opacity
   Fill_Z_Weapon_Color
  ) {materialFunction=&/MF_TopsSideProjection_Base}
  n38 = TextureObjectParameter() {
   group=TopSideProjection__Textures
   parameterName=TS_Normal
   samplerType=SAMPLERTYPE_Normal
   sortPriority=16
   texture=&FlatNormalBlue
  }
  n39 = LinearInterpolate(
   A=@n7.Normal
   B=Add(
    A=@n7.Normal
    B=MaterialFunctionCall!(
     Texture_Normals=@n38
     Rotation_Axis=@n21
     Caculated_Bounds=@n36
     Pre_Skinned_Normals=@n22
     Blend_Contrast=@n10
     Fill_alternate_Axis_Opacity=@n11
     Is_Vehicle=@n19
     Is_Mech=@n9
    ) {materialFunction=&/MF_TopsSideProjection_Normals}.Add_Surface_Normal
   )
   Alpha=ScalarParameter() {defaultValue=1, group=TopSideProjection, parameterName=TS_Normal_Intensty, sortPriority=1}
  )
  n40 = TextureObjectParameter() {
   group=TopSideProjection__Textures
   parameterName="TS_Mask_01 - R=X G=X B=X A=X"
   sortPriority=20
   texture=&T_BlacksRGB
  }
  n41 = VectorParameter() {group=TopSideProjection, parameterName=TS_Fill_alternate_Axis_Color_Mask_01, sortPriority=21, defaultValue=<rgba 0 0 0 1>}
  n42 = MaterialFunctionCall(
   Texture=@n40
   Rotation_Axis=@n21
   Caculated_Bounds=@n36
   Pre_Skinned_Normals=@n22
   Blend_Contrast=@n10
   Fill_alternate_Axis_Opacity=@n11
   "Fill_alternate_Axis_Color_A=Roughness"=AppendVector(A=@n41.RGB, B=@n41.RGB)
   Is_Vehicle=@n19
   Is_Mech=@n9
  ) {materialFunction=&MF_TopsSideProjection_Base}
  n43 = VectorParameter() {group=TopSideProjection_Emissive, parameterName=TS_Mask01_MaskChannel01_ForEmissive_Color, sortPriority=6, defaultValue=<rgba 1 1 1 1>}
  n44 = VectorParameter() {group=TopSideProjection_Emissive, parameterName=TS_Mask01_MaskChannel02_ForEmissive_Color, sortPriority=7, defaultValue=<rgba 0 0 0 1>}
  n45 = Add(
   A=Multiply(
    A=ChannelMaskParameter(None=@n42.Combined) {group=TopSideProjection_Emissive, parameterName=TS_Mask01_MaskChannel01_ForEmissive, sortPriority=6}.R
    B=@n43.RGB
   )
   B=Multiply(
    A=ChannelMaskParameter(None=@n42.Combined) {group=TopSideProjection_Emissive, parameterName=TS_Mask01_MaskChannel02_ForEmissive, sortPriority=7}.R
    B=@n44.RGB
   )
  )
  n46 = TextureObjectParameter() {
   group=TopSideProjection__Textures
   parameterName="TS_Warp01 - RGB=MaskedColor + RGBA=X for WarpUVs"
   sortPriority=4
   texture=&T_BlacksRGB
  }
  n47 = VectorParameter() {
   group=TopSideProjection_Emissive
   parameterName=TS-EmissiveRotation-Weapon-01---------------------
   sortPriority=160
   defaultValue=<rgba 0 0 -0.25 1>
  }
  n48 = VectorParameter() {group=TopSideProjection_Emissive, parameterName=TS_EmissiveRotation_Vehicle_01, sortPriority=161, defaultValue=<rgba 0 0 0 1>}
  n49 = VectorParameter() {
   desc=must activate TS_Vehicle_Side_01 UVs
   group=TopSideProjection_Emissive
   parameterName=TS_EmissiveRotation_Vehicle_Side_01
   sortPriority=163
   defaultValue=<rgba 0 0 0 1>
  }
  n50 = LinearInterpolate(A=@n48.RGB, B=@n49.RGB, Alpha=@n17)
  n51 = VectorParameter() {group=TopSideProjection_Emissive, parameterName=TS_EmissiveRotation_Mech_01, sortPriority=164, defaultValue=<rgba 0.25 0.25 0 1>}
  n52 = VectorParameter() {
   group=TopSideProjection_Emissive
   parameterName=TS_EmissivePanning_Weapon_01---------------------
   sortPriority=140
   defaultValue=<rgba 0 0 0 1>
  }
  n53 = VectorParameter() {group=TopSideProjection_Emissive, parameterName=TS_EmissivePanning_Vehicle_01, sortPriority=141, defaultValue=<rgba 0 0 0 1>}
  n54 = VectorParameter() {
   desc=must activate TS_Vehicle_Side_01 UVs
   group=TopSideProjection_Emissive
   parameterName=TS_EmissivePanning_Vehicle_Side_01
   sortPriority=141
   defaultValue=<rgba 0 0 0 1>
  }
  n55 = LinearInterpolate(A=@n52.RGB, B=LinearInterpolate(A=@n53.RGB, B=@n54.RGB, Alpha=@n17), Alpha=@n19)
  n56 = VectorParameter() {group=TopSideProjection_Emissive, parameterName=TS_EmissivePanning_Mech_01, sortPriority=142, defaultValue=<rgba 0 0 0 1>}
  n57 = Add(A=@n36, B=Multiply(A=Time(), B=LinearInterpolate(A=@n55, B=@n56.RGB, Alpha=@n9)))
  n58 = ScalarParameter() {defaultValue=1, group=TopSideProjection_Emissive, parameterName=TS_WarpUVS_01_TileRate Weapon, sortPriority=53}
  n59 = ScalarParameter() {defaultValue=1, group=TopSideProjection_Emissive, parameterName=TS_WarpUVS_01-TileRate Vehicle, sortPriority=54}
  n60 = VectorParameter() {group=TopSideProjection_Emissive, parameterName=TS_WarpUVS_01-Anitmated Weapon, sortPriority=57, defaultValue=<rgba 0 0 0 1>}
  n61 = VectorParameter() {group=TopSideProjection_Emissive, parameterName=TS_WarpUVS_01_Anitmated Vehicle, sortPriority=57, defaultValue=<rgba 0 0 0 1>}
  n62 = LinearInterpolate(A=@n60.RGB, B=@n61.RGB, Alpha=@n19)
  n63 = Add(
   A=Multiply(A=TextureCoordinate(), B=LinearInterpolate(A=@n58, B=@n59, Alpha=@n19))
   B=Multiply(A=Time() {period=1}, B=ComponentMask(None=@n62) {g=true, r=true})
  )
  n64 = ScalarParameter() {group=TopSideProjection_Emissive, parameterName=TS_WarpUVS_01-Amount Weapon, sortPriority=55}
  n65 = ScalarParameter() {group=TopSideProjection_Emissive, parameterName=TS_WarpUVS_01-Amount_Vehicle, sortPriority=56}
  n66 = ScalarParameter() {group=TopSideProjection_Emissive, parameterName=TS_WarpUVS-01-Amount Weapon_Reactive, sortPriority=206}
  n67 = ScalarParameter() {group=TopSideProjection_Emissive, parameterName=TS_WarpUVS-01-Amount_Vehicle_Reactive, sortPriority=206}
  n68 = Multiply(
   A=ChannelMaskParameter(None=TextureSample(UVs=@n63, Tex=@n46).RGBA) {group=TopSideProjection_Emissive, parameterName=TS-WarpUVS-01_Channel, sortPriority=201}.R
   B=LinearInterpolate(A=LinearInterpolate(A=@n64, B=@n65, Alpha=@n19), B=LinearInterpolate(A=@n66, B=@n67, Alpha=@n19), Alpha=NamedRerouteUsage())
  )
  n69 = MaterialFunctionCall(
   Texture=@n46
   Rotation_Axis=LinearInterpolate(A=LinearInterpolate(A=@n47.RGB, B=@n50, Alpha=@n19), B=@n51.RGB, Alpha=@n9)
   Caculated_Bounds=StaticSwitchParameter(True=Add(A=@n57, B=@n68), False=@n57) {defaultValue=true, group=TopSideProjection_Emissive, parameterName=Use Warp, sortPriority=200}
   Pre_Skinned_Normals=@n22
   Blend_Contrast=@n10
   Fill_alternate_Axis_Opacity=@n11
   "Fill_alternate_Axis_Color_A=Roughness"=@n13
   Is_Vehicle=@n19
   Is_Mech=@n9
  ) {materialFunction=&MF_TopsSideProjection_Base}
  n70 = Multiply(A=@n45, B=@n69.RGB)
  n71 = ScalarParameter() {defaultValue=1, group=TopSideProjection_Emissive, parameterName=TS-Emissive01-Brightness, sortPriority=0}
  n72 = SetMaterialAttributes!(
   MaterialAttributes=@n7.MaterialAttributes
   BaseColor=ComponentMask(None=@n37.Combined) {b=true, g=true, r=true}
   Roughness=LinearInterpolate(
    A=ScalarParameter() {group=TopSideProjection, parameterName=TS_Roughness_Min, sortPriority=13}
    B=ScalarParameter() {defaultValue=1, group=TopSideProjection, parameterName=TS_Roughness_Max, sortPriority=14}
    Alpha=ComponentMask(None=@n37.Combined) {a=true}
   )
   Normal=StaticSwitchParameter(True=@n7.Normal, False=ShadingPathSwitch(Default=@n39, Deferred=@n39, Mobile=@n7.Normal) {desc=Migrated from FeatureLevelSwitch}) {group=TopSideProjection, parameterName=No Normal, sortPriority=1}
   Specular=LinearInterpolate(
    A=ScalarParameter() {group=TopSideProjection, parameterName=TS_Specularty_Min, sortPriority=12}
    B=ScalarParameter() {defaultValue=1, group=TopSideProjection, parameterName=TS_Specularty_Max, sortPriority=12}
    Alpha=ChannelMaskParameter(None=@n42.Combined) {group=TopSideProjection, parameterName=TS_SpecChannel_Mask_01, sortPriority=26}.R
   )
   Metallic=ChannelMaskParameter(None=@n42.Combined) {group=TopSideProjection, parameterName=TS_MetalicChannel_Mask_01, sortPriority=26}.R
   EmissiveColor=Clamp(
    None=Multiply(A=@n70, B=@n71)
    Max=ScalarParameter() {defaultValue=25, group=TopSideProjection_Emissive, parameterName=TS-Emissive01-MaxBrightness, sortPriority=0}
   )
  ) {attributeSetTypes [
    #1
    #2
    #3
    #4
    #5
    #6
   ]}
  n73 = GetMaterialAttributes(None=BlendMaterialAttributes(A=@n7.MaterialAttributes, B=@n72, Alpha=LinearInterpolate(A=@n4, B=@n1, Alpha=@n5))) {attributeGetTypes [#6]}
  n74 = TextureObjectParameter() {
   group=TopSideProjection__Textures
   parameterName=TS_Emissive_01 - RGB texture
   sortPriority=5
   texture=&T_BlacksRGB
  }
  n75 = ScalarParameter() {group=TopSideProjection_Emissive, parameterName=TS_Fill_alternate_Axis_Opacity_Emissive01, sortPriority=0}
  n76 = VectorParameter() {group=TopSideProjection_Emissive, parameterName=TS_Fill_alternate_Axis_Color_Emissive01, sortPriority=1, defaultValue=<rgba 0 0 0 1>}
  n77 = AppendVector(A=@n76.RGB, B=@n76.RGB)
  n78 = Multiply(
   A=ComponentMask(None=MaterialFunctionCall(
     Texture=@n74
     Rotation_Axis=@n21
     Caculated_Bounds=@n36
     Pre_Skinned_Normals=@n22
     Blend_Contrast=@n10
     Fill_alternate_Axis_Opacity=@n75
     "Fill_alternate_Axis_Color_A=Roughness"=@n77
     Is_Vehicle=@n19
     Is_Mech=@n9
    ) {materialFunction=&MF_TopsSideProjection_Base}.Combined) {b=true, g=true, r=true}
   B=ScalarParameter() {defaultValue=1, group=Tomato_Reactive, parameterName=TS-Emissive01_ReactiveBrightness, sortPriority=5}
  )
  n79 = ScalarParameter() {defaultValue=2, group=Tomato_Reactive, parameterName=TS-BaseEmissiveBrightnes_Reactive, sortPriority=4}
  n80 = ChannelMaskParameter(None=@n42.Combined) {group=Tomato_Reactive, parameterName=TS_Mask01_MaskChannel_EmissiveReactive, sortPriority=2}
  n81 = VectorParameter() {group=Tomato_Reactive, parameterName=TS_Mask01_EmissiveReactive_Color, sortPriority=3, defaultValue=<rgba 0 0 0 1>}
  n82 = Multiply(A=Add(A=Multiply(A=@n45, B=@n79), B=Multiply(A=@n80.R, B=@n81.RGB)), B=@n69.RGB)
  n83 = MaterialFunctionCall(Color=Multiply(A=@n29.Result, B=Constant3Vector() {constant=<rgba 1 1 -1 0>})) {materialFunction=&SplitComponents}
  n84 = LinearInterpolate(A=AppendVector(A=@n83.G, B=@n83.G), B=AppendVector(A=@n83.R, B=@n83.R), Alpha=@n19)
  n85 = OneMinus(None=@n83.B)
  n86 = MaterialFunctionCall(UV Channel=LinearInterpolate(A=@n84, B=AppendVector(A=@n85, B=@n83.R), Alpha=@n9)) {materialFunction=&LinearGradient}
  n87 = ScalarParameter() {group=Tomato_Reactive, parameterName=TS_ReactiveGradientPosition, sortPriority=7}
  n88 = LinearInterpolate(
   A=ScalarParameter() {group=Tomato_Reactive, parameterName=TS_ReactiveGradientPosition_Vehicle, sortPriority=7}
   B=ScalarParameter() {group=Tomato_Reactive, parameterName=TS_ReactiveGradientPosition_VehicleSide, sortPriority=7}
   Alpha=NamedRerouteUsage()
  )
  n89 = NamedRerouteUsage()
  n90 = LinearInterpolate(
   A=LinearInterpolate(
    A=LinearInterpolate(A=@n87, B=@n88, Alpha=@n89)
    B=ScalarParameter() {group=Tomato_Reactive, parameterName=TS_ReactiveGradientPosition_Mech, sortPriority=7}
    Alpha=NamedRerouteUsage()
   )
   B=ScalarParameter() {group=Tomato_Reactive, parameterName=TS_ReactiveGradientPosition_Character, sortPriority=7}
   Alpha=NamedRerouteUsage()
  )
  n91 = VectorParameter() {group=Tomato_Reactive, parameterName=TS_Emissive_ReactiveGradient_Color-----, sortPriority=6, defaultValue=<rgba 1 1 1 1>}
  n92 = Multiply(
   A=Power(
    Base=Multiply(A=@n86.UGradient, B=@n90) {constB=0.8}
    Exp=ScalarParameter() {defaultValue=15, group=Tomato_Reactive, parameterName=TS_ReactiveGradientPower, sortPriority=7}
   ) {constExponent=7.5}
   B=@n91.RGB
  )
  n93 = Multiply(
   A=LinearInterpolate(B=Add(A=@n82, B=@n92), Alpha=NamedRerouteUsage())
   B=LinearInterpolate(
    B=ScalarParameter() {defaultValue=1, group=Tomato_Reactive, parameterName=TS-Emissive01_ReactiveWarp_BrightnessMultiplier, sortPriority=6}
    Alpha=NamedRerouteUsage()
   )
  )
  n94 = Add(A=@n73.EmissiveColor, B=LinearInterpolate(B=Add(A=@n78, B=@n93), Alpha=NamedRerouteUsage()))
  n95 = LinearInterpolate(
   A=@n94
   B=Multiply(A=@n4, B=@n94)
   Alpha=ScalarParameter() {group=Tomato_Reactive, parameterName=TS_MaskReactiveEmissive-----, sortPriority=-75}
  )
  n96 = StaticSwitchParameter(
   True=SetMaterialAttributes!(MaterialAttributes=@n73.MaterialAttributes, EmissiveColor=@n95) {attributeSetTypes [#6]}
   False=@n6
  ) {
   desc=Topside Projection - Ambient glow with mask and texture - Reactive - Emissive glow and front gradient on fire
   group=Customization_Wrap
   parameterName=Wrap_Tomato
  }
  FunctionOutput(None=@n96)
  n97 = OneMinus(None=@n83.B)
  LinearInterpolate(A=@n84, B=Multiply(A=AppendVector(A=@n97, B=@n83.R)) {constB=0.425}, Alpha=@n9)
  LinearInterpolate(A=@n95, B=@n1, Alpha=NamedRerouteUsage())
  NamedRerouteDeclaration(None=FunctionInput() {bUsePreviewValueAsDefault=true, inputName=IsCharacter, inputType=FunctionInput_Scalar, sortPriority=3}) {name=IsCharacter, nodeColor=<rgba 0.4766 1 0 1>}
  NamedRerouteDeclaration(None=@n17) {name=IsVehicleSide, nodeColor=<rgba 1 0 0.7969 1>}
  NamedRerouteDeclaration(None=@n19) {name=IsVehicle, nodeColor=<rgba 0 0.0859 1 1>}
  NamedRerouteDeclaration(None=FunctionInput() {bUsePreviewValueAsDefault=true, inputName=isCharAccent, inputType=FunctionInput_Scalar, sortPriority=78}) {name=isCharAccent, nodeColor=<rgba 0 0.9253 1 1>}
  NamedRerouteDeclaration(None=@n9) {name=IsMech, nodeColor=<rgba 1 0.5625 0 1>}
  n98 = Subtract(A=Time(), B=ScalarParameter() {group=00-Reactive-GTFB, parameterName=00-GameTimeFromBP, sortPriority=100})
  n99 = ScalarParameter() {defaultValue=1, group=Tomato_Reactive, parameterName=TO_AnimationDuration-----, sortPriority=100}
  n100 = Power(
   Base=Saturate(None=Divide(A=@n98, B=@n99))
   Exp=ScalarParameter() {defaultValue=0.5, group=Tomato_Reactive, parameterName=TO_Curve_Amount, sortPriority=100}
  )
  n101 = ScalarParameter() {group=Tomato_Reactive, parameterName=TO_TestReactive, sortPriority=300}
  NamedRerouteDeclaration(None=LinearInterpolate(A=OneMinus(None=@n100) {desc=inverts curve from 0-1 to 1-0}, B=@n101, Alpha=@n101)) {name=Reactive, nodeColor=<rgba 0 1 0.7422 1>}
  Reroute(None=@n29.Result)
 ]}
