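# Material function for weapon wrap skins that projects two texture layers onto weapons, pickaxes, backpacks, vehicles, mechs and characters.
# Per-type offset, scale, panning and rotation parameters, chosen by lerps, transform bounds-relative UVs, with optional warping.
# Static switches pick top/side blending from pre-skinned normals or side projection for color and mask textures.
# Emissive masks combine those masks with bounds gradients, including time-triggered reactive fades.
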MaterialFunction /Game/Weapons/WeaponSkins/Wraps/MaterialFunctions/MF_TopSideProjection_Wrap
graph {nodes [
  n1 = Constant3Vector() {constant=<rgba 1 1 0 0>}
  n2 = GetMaterialAttributes(None=FunctionInput() {inputName=Material_Bypass, inputType=FunctionInput_MaterialAttributes, sortPriority=1}) {attributeGetTypes [#1, #2, #3]}
  n3 = VectorParameter() {group=TopSideProjection_UVS, parameterName=TS_Offset_Weapon_01---------------------, sortPriority=10, defaultValue=<rgba 0 0 0 1>}
  n4 = VectorParameter() {group=TopSideProjection_UVS, parameterName=TS_Offset_Pickaxe_01---------------------, sortPriority=10, defaultValue=<rgba 0 0 0 1>}
  n5 = LinearInterpolate(A=@n3.RGB, B=@n4.RGB, Alpha=NamedRerouteUsage())
  n6 = VectorParameter() {group=TopSideProjection_UVS, parameterName=TS_Offset_Vehicle_01, sortPriority=11, defaultValue=<rgba 0 0 0 1>}
  n7 = VectorParameter() {
   desc=must activate TS_Vehicle_Side_01 UVs
   group=TopSideProjection_UVS
   parameterName=TS_Offset_Vehicle_Side_01
   sortPriority=11
   defaultValue=<rgba 0 0 0 1>
  }
  n8 = LinearInterpolate(A=@n6.RGB, B=@n7.RGB, Alpha=NamedRerouteUsage())
  n9 = MaterialFunctionCall(Float4=NamedRerouteUsage()) {materialFunction=&BreakOutFloat4Components}
  n10 = VectorParameter() {group=TopSideProjection_UVS, parameterName=TS_Offset_Mech_01, sortPriority=12, defaultValue=<rgba 0 0 0 1>}
  n11 = VectorParameter() {group=TopSideProjection_UVS, parameterName=TS_Offset_Character_01, sortPriority=12, defaultValue=<rgba 0 0 0 1>}
  n12 = VectorParameter() {group=TopSideProjection_UVS, parameterName=TS_Offset_Backpack_01---------------------, sortPriority=10, defaultValue=<rgba 0 0 0 1>}
  n13 = LinearInterpolate(
   A=LinearInterpolate(A=LinearInterpolate(A=@n5, B=@n8, Alpha=@n9.R), B=LinearInterpolate(A=@n10.RGB, B=@n11.RGB, Alpha=@n9.A), Alpha=@n9.B)
   B=@n12.RGB
   Alpha=NamedRerouteUsage()
  )
  n14 = FunctionInput(Preview=Constant()) {inputName=isVehicle, inputType=FunctionInput_Scalar, sortPriority=100}
  n15 = FunctionInput(Preview=Constant()) {bUsePreviewValueAsDefault=true, inputName=isMech, inputType=FunctionInput_Scalar, sortPriority=100}
  n16 = FunctionInput(Preview=VertexInterpolator(VS=PreSkinnedPosition()).PS) {bUsePreviewValueAsDefault=true, inputName=Local Position, sortPriority=11}
  n17 = MaterialFunctionCall!(
   is Vehicle=@n14
   Object Bounds Max Adjustment=FunctionInput() {bUsePreviewValueAsDefault=true, inputName=Object Bounds Max Adjustment, sortPriority=13}
   Object Bounds Min Adjustment=FunctionInput() {bUsePreviewValueAsDefault=true, inputName=Object Bounds Min Adjustment, sortPriority=12}
   Object Bounds Center Adjustment=FunctionInput() {bUsePreviewValueAsDefault=true, inputName=Object Bounds Center Adjustment, sortPriority=14}
   isMech=@n15
   Side Project Vehicle
   Local Position=@n16
  ) {materialFunction=&/MF_WrapBounds}
  n18 = Time()
  n19 = VectorParameter() {group=TopSideProjection_UVS, parameterName=TS_Panning_Weapon_01---------------------, sortPriority=40, defaultValue=<rgba 0 0 0 1>}
  n20 = VectorParameter() {group=TopSideProjection_UVS, parameterName=TS_Panning_Vehicle_01, sortPriority=41, defaultValue=<rgba 0 0 0 1>}
  n21 = VectorParameter() {
   desc=must activate TS_Vehicle_Side_01 UVs
   group=TopSideProjection_UVS
   parameterName=TS_Panning_Vehicle_Side_01
   sortPriority=41
   defaultValue=<rgba 0 0 0 1>
  }
  n22 = NamedRerouteUsage()
  n23 = MaterialFunctionCall(Float4=NamedRerouteUsage()) {materialFunction=&BreakOutFloat4Components}
  n24 = VectorParameter() {group=TopSideProjection_UVS, parameterName=TS_Panning_Mech_01, sortPriority=42, defaultValue=<rgba 0 0 0 1>}
  n25 = VectorParameter() {group=TopSideProjection_UVS, parameterName=TS_Panning_Character_01, sortPriority=42, defaultValue=<rgba 0 0 0 1>}
  n26 = LinearInterpolate(
   A=LinearInterpolate(A=@n19.RGB, B=LinearInterpolate(A=@n20.RGB, B=@n21.RGB, Alpha=@n22), Alpha=@n23.R)
   B=LinearInterpolate(A=@n24.RGB, B=@n25.RGB, Alpha=@n23.A)
   Alpha=@n23.B
  )
  n27 = VectorParameter() {group=TopSideProjection_UVS, parameterName="TS_Scale_Weapon_01 A=uniform", sortPriority=30, defaultValue=<rgba 1 1 1 1>}
  n28 = VectorParameter() {group=TopSideProjection_UVS, parameterName="TS_Scale_Vehicle_01 A=uniform", sortPriority=31, defaultValue=<rgba 1 1 1 1>}
  n29 = Multiply(A=@n28.RGB, B=@n28.RGB)
  n30 = VectorParameter() {
   desc=must activate TS_Vehicle_Side_01 UVs
   group=TopSideProjection_UVS
   parameterName="TS_Scale_Vehicle_Side_01 A=uniform"
   sortPriority=31
   defaultValue=<rgba 1 1 1 1>
  }
  n31 = Multiply(A=@n30.RGB, B=@n30.RGB)
  n32 = NamedRerouteUsage()
  n33 = MaterialFunctionCall(Float4=NamedRerouteUsage()) {materialFunction=&BreakOutFloat4Components}
  n34 = VectorParameter() {group=TopSideProjection_UVS, parameterName="TS_Scale_Mech_01 A=uniform", defaultValue=<rgba 1 1 1 1>}
  n35 = VectorParameter() {group=TopSideProjection_UVS, parameterName="TS_Scale_Character_01 A=uniform", defaultValue=<rgba 1 1 1 1>}
  n36 = LinearInterpolate(
   A=LinearInterpolate(A=Multiply(A=@n27.RGB, B=@n27.RGB), B=LinearInterpolate(A=@n29, B=@n31, Alpha=@n32), Alpha=@n33.R)
   B=LinearInterpolate(A=Multiply(A=@n34.RGB, B=@n34.RGB), B=Multiply(A=@n35.RGB, B=@n35.RGB), Alpha=@n33.A)
   Alpha=@n33.B
  )
  n37 = VectorParameter() {group=TopSideProjection_UVS, parameterName="TS_Scale_Pickaxe_01 A=uniform", sortPriority=30, defaultValue=<rgba 1 1 1 1>}
  n38 = Multiply(A=@n37.RGB, B=@n37.RGB)
  n39 = NamedRerouteUsage()
  n40 = VectorParameter() {group=TopSideProjection_UVS, parameterName="TS_Scale_Backpack_01 A=uniform", sortPriority=30, defaultValue=<rgba 1 1 1 1>}
  n41 = Multiply(
   A=Add(A=Add(A=@n17.Result, B=@n13), B=Multiply(A=@n18, B=@n26))
   B=LinearInterpolate(A=LinearInterpolate(A=@n36, B=@n38, Alpha=@n39), B=Multiply(A=@n40.RGB, B=@n40.RGB), Alpha=NamedRerouteUsage())
  )
  Add(
   A=Multiply(
    A=Multiply(A=@n2.Normal, B=@n1)
    B=ScalarParameter() {group=TopSideProjection_UVS, parameterName=TS-SurfaceNormal_WarpUVS_01_Amount, sortPriority=55}
   )
   B=@n41
  )
  n42 = Constant3Vector() {constant=<rgba 1 1 0 0>}
  n43 = VectorParameter() {group=TopSideProjection_UVS, parameterName=TS_Offset_Weapon_02---------------------, sortPriority=160, defaultValue=<rgba 0 0 0 1>}
  n44 = VectorParameter() {group=TopSideProjection_UVS, parameterName=TS_Offset_Vehicle_02, sortPriority=161, defaultValue=<rgba 0 0 0 1>}
  n45 = VectorParameter() {
   desc=must activate TS_Vehicle_Side_02 UVs
   group=TopSideProjection_UVS
   parameterName=TS_Offset_Vehicle_Side_02
   sortPriority=161
   defaultValue=<rgba 0 0 0 1>
  }
  n46 = NamedRerouteUsage()
  n47 = MaterialFunctionCall(Float4=NamedRerouteUsage()) {materialFunction=&BreakOutFloat4Components}
  n48 = VectorParameter() {group=TopSideProjection_UVS, parameterName=TS_Offset_Mech_02, sortPriority=162, defaultValue=<rgba 0 0 0 1>}
  n49 = VectorParameter() {group=TopSideProjection_UVS, parameterName=TS_Offset_Character_02, sortPriority=162, defaultValue=<rgba 0 0 0 1>}
  n50 = LinearInterpolate(
   A=LinearInterpolate(A=@n43.RGB, B=LinearInterpolate(A=@n44.RGB, B=@n45.RGB, Alpha=@n46), Alpha=@n47.R)
   B=LinearInterpolate(A=@n48.RGB, B=@n49.RGB, Alpha=@n47.A)
   Alpha=@n47.B
  )
  n51 = VectorParameter() {group=TopSideProjection_UVS, parameterName=TS_Offset_Pickaxe_02---------------------, sortPriority=10, defaultValue=<rgba 0 0 0 1>}
  n52 = NamedRerouteUsage()
  n53 = VectorParameter() {group=TopSideProjection_UVS, parameterName=TS_Offset_Backpack_02---------------------, sortPriority=10, defaultValue=<rgba 0 0 0 1>}
  n54 = Add(A=@n17.Result, B=LinearInterpolate(A=LinearInterpolate(A=@n50, B=@n51.RGB, Alpha=@n52), B=@n53.RGB, Alpha=NamedRerouteUsage()))
  n55 = VectorParameter() {group=TopSideProjection_UVS, parameterName="TS_ScaleUV/Blend_Weapon_02 A=uniform", sortPriority=170, defaultValue=<rgba 1 1 1 1>}
  n56 = Multiply(A=@n55.RGB, B=@n55.RGB)
  n57 = VectorParameter() {group=TopSideProjection_UVS, parameterName="TS_ScaleUV/Blend_Vehicle_02 A=uniform", sortPriority=171, defaultValue=<rgba 1 1 1 1>}
  n58 = Multiply(A=@n57.RGB, B=@n57.RGB)
  n59 = MaterialFunctionCall(Float4=NamedRerouteUsage()) {materialFunction=&BreakOutFloat4Components}
  n60 = VectorParameter() {group=TopSideProjection_UVS, parameterName="TS_ScaleUV/Blend_Mech_02 A=uniform", sortPriority=172, defaultValue=<rgba 1 1 1 1>}
  n61 = VectorParameter() {group=TopSideProjection_UVS, parameterName="TS_Scale_Weapon_02 A=uniform", sortPriority=170, defaultValue=<rgba 1 1 1 1>}
  n62 = VectorParameter() {group=TopSideProjection_UVS, parameterName="TS_Scale_Vehicle_02 A=uniform", sortPriority=171, defaultValue=<rgba 1 1 1 1>}
  n63 = Multiply(A=@n62.RGB, B=@n62.RGB)
  n64 = VectorParameter() {
   desc=must activate TS_Vehicle_Side_02 UVs
   group=TopSideProjection_UVS
   parameterName="TS_Scale_Vehicle_Side_02 A=uniform"
   sortPriority=171
   defaultValue=<rgba 1 1 1 1>
  }
  n65 = Multiply(A=@n64.RGB, B=@n64.RGB)
  n66 = NamedRerouteUsage()
  n67 = MaterialFunctionCall(Float4=NamedRerouteUsage()) {materialFunction=&BreakOutFloat4Components}
  n68 = VectorParameter() {group=TopSideProjection_UVS, parameterName="TS_Scale_Mech_02 A=uniform", sortPriority=172, defaultValue=<rgba 1 1 1 1>}
  n69 = VectorParameter() {group=TopSideProjection_UVS, parameterName="TS_Scale_Character_02 A=uniform", sortPriority=172, defaultValue=<rgba 1 1 1 1>}
  n70 = LinearInterpolate(
   A=LinearInterpolate(A=Multiply(A=@n61.RGB, B=@n61.RGB), B=LinearInterpolate(A=@n63, B=@n65, Alpha=@n66), Alpha=@n67.R)
   B=LinearInterpolate(A=Multiply(A=@n68.RGB, B=@n68.RGB), B=Multiply(A=@n69.RGB, B=@n69.RGB), Alpha=@n67.A)
   Alpha=@n67.B
  )
  n71 = ScalarParameter() {group=TopSideProjection_UVS, parameterName="TS_UV_Blend_OffsetWeapon (Only02)", sortPriority=300}
  n72 = ScalarParameter() {group=TopSideProjection_UVS, parameterName="TS_UV_Blend_OffsetVehicle (Only02)", sortPriority=301}
  n73 = MaterialFunctionCall(Float4=NamedRerouteUsage()) {materialFunction=&BreakOutFloat4Components}
  n74 = ScalarParameter() {defaultValue=1, group=TopSideProjection_UVS, parameterName="TS_UV_Blend_FalloffWeapon (Only02) negative inverts", sortPriority=305}
  n75 = ScalarParameter() {defaultValue=1, group=TopSideProjection_UVS, parameterName="TS_UV_Blend_FalloffVehicle (Only02)", sortPriority=306}
  n76 = FunctionInput(Preview=Constant3Vector() {constant=<rgba 1 0 0 0>}) {bUsePreviewValueAsDefault=true, inputName=ForwardAxis}
  n77 = MaterialFunctionCall!(
   Bounds=@n17.Result
   ForwardAxis=@n76
   Offset=LinearInterpolate(
    A=LinearInterpolate(A=@n71, B=@n72, Alpha=@n73.R)
    B=ScalarParameter() {group=TopSideProjection_UVS, parameterName="TS__UV_Blend_OffsetMehicle (Only02)", sortPriority=302}
    Alpha=@n73.B
   )
   FallOff=LinearInterpolate(
    A=LinearInterpolate(A=@n74, B=@n75, Alpha=@n73.R)
    B=ScalarParameter() {group=TopSideProjection_UVS, parameterName="TS_UV_Blend_FalloffMech (Only02)", sortPriority=307}
    Alpha=@n73.B
   )
  ) {materialFunction=&/MF_Wrap_Gradent_Simple}
  n78 = LinearInterpolate(
   A=LinearInterpolate(A=LinearInterpolate(A=@n56, B=@n58, Alpha=@n59.R), B=Multiply(A=@n60.RGB, B=@n60.RGB), Alpha=@n59.B)
   B=@n70
   Alpha=@n77.Clamped_Data
  )
  n79 = VectorParameter() {group=TopSideProjection_UVS, parameterName=TS_Panning_Weapon_02---------------------, sortPriority=180, defaultValue=<rgba 0 0 0 1>}
  n80 = VectorParameter() {group=TopSideProjection_UVS, parameterName=TS_Panning_Vehicle_02, sortPriority=181, defaultValue=<rgba 0 0 0 1>}
  n81 = VectorParameter() {
   desc=must activate TS_Vehicle_Side_02 UVs
   group=TopSideProjection_UVS
   parameterName=TS_Panning_Vehicle_Side_02
   sortPriority=181
   defaultValue=<rgba 0 0 0 1>
  }
  n82 = LinearInterpolate(A=@n80.RGB, B=@n81.RGB, Alpha=NamedRerouteUsage())
  n83 = MaterialFunctionCall(Float4=NamedRerouteUsage()) {materialFunction=&BreakOutFloat4Components}
  n84 = VectorParameter() {group=TopSideProjection_UVS, parameterName=TS_Panning_Mech_02, sortPriority=182, defaultValue=<rgba 0 0 0 1>}
  n85 = VectorParameter() {group=TopSideProjection_UVS, parameterName=TS_Panning_Character_02, sortPriority=182, defaultValue=<rgba 0 0 0 1>}
  n86 = Multiply(
   A=Time()
   B=LinearInterpolate(A=LinearInterpolate(A=@n79.RGB, B=@n82, Alpha=@n83.R), B=LinearInterpolate(A=@n84.RGB, B=@n85.RGB, Alpha=@n83.A), Alpha=@n83.B)
  )
  n87 = StaticSwitch(
   True=Add(A=Multiply(A=@n54, B=@n78), B=@n86)
   False=Multiply(A=Add(A=@n54, B=@n86), B=@n70)
   Value=StaticBoolParameter() {group=TopSideProjection, parameterName=TS-Blend-UV/Transition-Split-02------------------, sortPriority=299}
  )
  Add(
   A=Multiply(
    A=Multiply(A=@n2.Normal, B=@n42)
    B=ScalarParameter() {group=TopSideProjection_UVS, parameterName=TS-SurfaceNormal_WarpUVS_02_Amount, sortPriority=205}
   )
   B=@n87
  )
  ComponentMask() {g=true, r=true}
  n88 = Subtract(
   A=Time()
   B=Add(
    A=ScalarParameter() {group=TopSideProjection_Emissive_Reactive, parameterName=TS_Emissive01_Time_BP_Delay_Brightness}
    B=ScalarParameter() {defaultValue=-9999999, group=TopSideProjection_Emissive_Reactive, parameterName=TS_Emissive01_Time_BP_Brightness}
   )
  )
  n89 = ScalarParameter() {defaultValue=0.25, group=TopSideProjection_Emissive_Reactive, parameterName=TS_Emissive01_Time_BP_Duration_Brightness}
  n90 = Power(
   Base=Saturate(None=Divide(A=@n88, B=@n89))
   Exp=ScalarParameter() {defaultValue=0.25, group=TopSideProjection_Emissive_Reactive, parameterName=TS_Emissive01_Time_BP_Falloff}
  )
  n91 = LinearInterpolate(
   A=ScalarParameter() {group=TopSideProjection_Emissive_Reactive, parameterName=TS_Emissive01_Time_Brightness_DarkenBase_Start}
   B=ScalarParameter() {group=TopSideProjection_Emissive_Reactive, parameterName=TS_Emissive01_Time_Brightness_DarkenBase_End}
   Alpha=@n90
  )
  n92 = ScalarParameter() {group=TopSideProjection_Emissive, parameterName=TS-Emissive01-DarkenBase-------, sortPriority=4}
  n93 = ScalarParameter() {group=TopSideProjection_Emissive_Reactive, parameterName=TS_Emissive02/effect Emissive01_Gradent_Start, sortPriority=210}
  n94 = VectorParameter() {group=TopSideProjection_Emissive, parameterName=TS_Emissive01_OffsetWeapon, sortPriority=25, defaultValue=<rgba 0 -1 0 0>}
  n95 = StaticBoolParameter() {group=TopSideProjection_Emissive, parameterName="Ts-Emissive01_Gradent-->Texture", sortPriority=10}
  n96 = ScalarParameter() {defaultValue=0.5, group=TopSideProjection_Emissive, parameterName="TS_Emissive01_FalloffWeapon (negative inverts)", sortPriority=23}
  n97 = VectorParameter() {group=TopSideProjection_Emissive, parameterName=TS_Emissive01_OffsetVehicle, sortPriority=26, defaultValue=<rgba -1 0 0 0>}
  n98 = VectorParameter() {group=TopSideProjection_Emissive, parameterName=TS_Emissive01_OffsetMech, sortPriority=27, defaultValue=<rgba 0 0 -1 0>}
  n99 = VectorParameter() {group=TopSideProjection_Emissive, parameterName=TS_Emissive01_OffsetCharacter, sortPriority=27, defaultValue=<rgba 0 0 -1 0>}
  n100 = MaterialFunctionCall(Float4=NamedRerouteUsage()) {materialFunction=&BreakOutFloat4Components}
  n101 = ScalarParameter() {defaultValue=-0.5, group=TopSideProjection_Emissive, parameterName="TS_Emissive01_OffsetMech ", sortPriority=22}
  n102 = ScalarParameter() {defaultValue=-0.5, group=TopSideProjection_Emissive, parameterName=TS_Emissive01_OffsetCharacter_, sortPriority=22}
  n103 = VectorParameter() {group=TopSideProjection_Emissive, parameterName=TS_Emissive01_Spread_XY, sortPriority=21, defaultValue=<rgba 1 1 1 0>}
  n104 = VectorParameter() {group=TopSideProjection_Emissive, parameterName=TS_Emissive01_Spread_Mech_XY, sortPriority=20, defaultValue=<rgba 1 1 1 0>}
  n105 = VectorParameter() {group=TopSideProjection_Emissive, parameterName=TS_Emissive01_Animated_WeaponXY, sortPriority=30}
  n106 = VectorParameter() {group=TopSideProjection_Emissive, parameterName=TS_Emissive01_Animated_VehicleXY, sortPriority=31}
  n107 = VectorParameter() {group=TopSideProjection_Emissive, parameterName=TS_Emissive01_Animated_MechXY}
  n108 = VectorParameter() {group=TopSideProjection_Emissive, parameterName=TS_Emissive01_Animated_CharacterXY}
  n109 = TextureObjectParameter() {
   group=TopSideProjection
   parameterName=TS_Mask_02
   samplerType=SAMPLERTYPE_Masks
   sortPriority=62
   texture=&T_BlackMask
  }
  n110 = LinearInterpolate(
   A=ScalarParameter() {defaultValue=1, group=TopSideProjection, parameterName=TS_Blend_Contrast, sortPriority=-1}
   B=ScalarParameter() {defaultValue=2, group=TopSideProjection, parameterName=TS_Blend_Contrast_Mech, sortPriority=-1}
   Alpha=@n15
  )
  n111 = ScalarParameter() {group=TopSideProjection, parameterName=TS_Fill_alternate_Axis_Opacity 02, sortPriority=51}
  n112 = Max(A=@n111, B=ScalarParameter() {group=TopSideProjection, parameterName=TS_Fill_alternate_Axis_Opacity 02 VehicleSide, sortPriority=51})
  n113 = NamedRerouteUsage()
  n114 = LinearInterpolate(
   A=StaticSwitch(
    True=Constant() {r=1}
    False=LinearInterpolate(A=@n111, B=@n112, Alpha=@n113)
    Value=StaticBoolParameter() {group=TopSideProjection, parameterName=TS02--USE-Miplevels-VS-Alternative_Axies, sortPriority=50}
   )
   B=ScalarParameter() {group=TopSideProjection, parameterName=TS_Fill_alternate_Axis_Opacity_Character 02, sortPriority=51}
   Alpha=NamedRerouteUsage()
  )
  n115 = VectorParameter() {group=TopSideProjection, parameterName=TS_Fill_alternate_Axis_Color_Mask 02, sortPriority=67, defaultValue=<rgba 0 0 0 1>}
  n116 = AppendVector(A=@n115.RGB, B=@n115.RGB)
  n117 = VectorParameter() {group=TopSideProjection, parameterName=TS_Fill_Z_Axis_Weapon_Color_Mask_02, sortPriority=67, defaultValue=<rgba 0 0 0 1>}
  n118 = VectorParameter() {group=TopSideProjection_UVS, parameterName=TS-Rotation-Weapon-02---------------------, sortPriority=150, defaultValue=<rgba 0 0 -0.25 1>}
  n119 = VectorParameter() {group=TopSideProjection_UVS, parameterName=TS_Rotation_Vehicle_02, sortPriority=151, defaultValue=<rgba 0 0 0 1>}
  n120 = VectorParameter() {
   desc=must activate TS_Vehicle_Side_02 UVs
   group=TopSideProjection_UVS
   parameterName=TS_Rotation_Vehicle_Side_02
   sortPriority=151
   defaultValue=<rgba 0 0 0 1>
  }
  n121 = MaterialFunctionCall(Float4=NamedRerouteUsage()) {materialFunction=&BreakOutFloat4Components}
  n122 = NamedRerouteDeclaration(None=Multiply(A=@n121.G, B=ScalarParameter() {group=TopSideProjection_UVS, parameterName="TS_Vehicle_Side_02 UVs  0=Off 1=Active", sortPriority=149})) {name=TS_Vehicle_Side_02 ON, nodeColor=<rgba 1 0.5966 0 1>}
  n123 = VectorParameter() {group=TopSideProjection_UVS, parameterName=TS_Rotation_Mech_02, sortPriority=153, defaultValue=<rgba 0.25 0.25 0 1>}
  n124 = VectorParameter() {group=TopSideProjection_UVS, parameterName=TS_Rotation_Character_02, sortPriority=153, defaultValue=<rgba 0.25 0.25 0 1>}
  n125 = LinearInterpolate(
   A=LinearInterpolate(A=@n118.RGB, B=LinearInterpolate(A=@n119.RGB, B=@n120.RGB, Alpha=@n122), Alpha=@n121.R)
   B=LinearInterpolate(A=@n123.RGB, B=@n124.RGB, Alpha=@n121.A)
   Alpha=@n121.B
  )
  n126 = StaticSwitchParameter(True=Add(A=@n87, B=NamedRerouteUsage()), False=@n87) {group=TopSideProjection_UVS, parameterName=TS-WarpUVS-02---------------------, sortPriority=200}
  n127 = FunctionInput(Preview=VertexInterpolator(VS=PreSkinnedNormal()).PS) {bUsePreviewValueAsDefault=true, inputName=Pre-Skinned Local Normal, sortPriority=10}
  n128 = StaticBoolParameter() {group=TopSideProjection, parameterName=TS-Use-Side-Projectiion-02, sortPriority=55}
  n129 = TextureObjectParameter() {
   group=TopSideProjection
   parameterName=TS_Normal_02
   samplerType=SAMPLERTYPE_Normal
   sortPriority=61
   texture=&FlatNormalBlue
  }
  n130 = MaterialFunctionCall(Float4=NamedRerouteUsage()) {materialFunction=&BreakOutFloat4Components}
  n131 = MaterialFunctionCall!(
   For_Safe_Zones
   Normal_Map=@n129
   Caculated_Bounds=@n17.Result
   Pre_Skinned_Normals=@n127
   Is Vehicle=@n130.R
   Is VehicleSide=@n130.R
   Is Mech=@n130.R
   Vehicle_Rotation
   UV_Rotation Bypass=NamedRerouteUsage()
   Mech_Rotation
   Uvs_Offset=NamedRerouteUsage()
   Offset-after rotation
   Uvs_TileRate=NamedRerouteUsage()
  ) {materialFunction=&/MF_Projection_Side_Normal}
  n132 = NamedRerouteUsage()
  n133 = ScalarParameter() {defaultValue=3, group=TopSideProjection, parameterName=TS02_Mip_levels, sortPriority=50}
  n134 = TextureSample!(UVs=@n131.UVs, Tex=@n132, Level=@n133, Apply View MipBias) {mipValueMode=TMVM_MipLevel}
  n135 = TextureSample!(UVs=@n131.UVs, Tex=@n132, Apply View MipBias)
  n136 = NamedRerouteUsage()
  n137 = NamedRerouteUsage()
  n138 = StaticBoolParameter() {group=TopSideProjection, parameterName=TS02--USE-Miplevels-VS-Alternative_Axies, sortPriority=50}
  n139 = StaticSwitch(
   True=StaticSwitch(True=LinearInterpolate(A=@n134.RGBA, B=@n135.RGBA, Alpha=@n136), False=LinearInterpolate(A=@n116, B=@n135.RGBA, Alpha=@n137), Value=@n138)
   False=MaterialFunctionCall!(
    IsCharacter
    IsStatsB
    Texture=NamedRerouteDeclaration(None=@n109) {name=TS_Mask_02, nodeColor=<rgba 1 0 0.5391 1>}
    Rotation_Axis=@n125
    Caculated_Bounds=@n126
    Pre_Skinned_Normals=@n127
    Blend_Contrast=@n110
    Fill_alternate_Axis_Opacity=@n114
    "Fill_alternate_Axis_Color_A=Roughness"=@n116
    Is_Vehicle=@n14
    Is_Mech=@n15
    Enable_Fill_Z_Weapon=StaticBool() {value=true}
    Fill_Z_Weapon_opacity=ScalarParameter() {group=TopSideProjection, parameterName=TS_Fill_Z_Axis_Weapon_Opacity 02, sortPriority=51}
    Fill_Z_Weapon_Color=AppendVector(A=@n117.RGB, B=@n117.RGB)
   ) {materialFunction=&/MF_TopsSideProjection_Base}.Combined
   Value=@n128
  )
  n140 = MaterialFunctionCall(Float4=NamedRerouteUsage()) {materialFunction=&BreakOutFloat4Components}
  n141 = NamedRerouteUsage()
  n142 = NamedRerouteUsage()
  n143 = NamedRerouteUsage()
  n144 = MaterialFunctionCall!(
   Normal_Map=NamedRerouteUsage()
   Caculated_Bounds=NamedRerouteUsage()
   Pre_Skinned_Normals=NamedRerouteUsage()
   Is Vehicle=@n140.R
   Is Vehicle side=@n140.R
   Is Mech=@n140.R
   Offset-after rotation
   UV_Rotation_Right=@n141
   UV_Rotation_Left=Multiply(A=@n141) {constB=-1}
   Uvs_TileRate_Right=@n142
   Uvs_TileRate_Left=@n142
   Uvs_Offset_Right=@n143
   Uvs_Offset_Left=Multiply(A=@n143, B=Constant3Vector() {constant=<rgba -1 1 1 0>})
  ) {materialFunction=&/MF_Projection_Side_Left_Right_Decal_Normal}
  n145 = MaterialFunctionCall(Float4=NamedRerouteUsage()) {materialFunction=&BreakOutFloat4Components}
  n146 = MaterialFunctionCall(
   Normal_Map=NamedRerouteUsage()
   Caculated_Bounds=NamedRerouteUsage()
   Pre_Skinned_Normals=NamedRerouteUsage()
   Is Vehicle=@n145.R
   Is VehicleSide=@n145.R
   Is Mech=@n145.R
   UV_Rotation Bypass=NamedRerouteUsage()
   Uvs_Offset=NamedRerouteUsage()
   Uvs_TileRate=NamedRerouteUsage()
  ) {materialFunction=&MF_Projection_Side_Normal}
  n147 = StaticBoolParameter() {group=TopSideProjection, parameterName=TS-Use-SP-Left-Right-01, sortPriority=10}
  n148 = StaticSwitch(True=@n144.UVs, False=@n146.UVs, Value=@n147)
  n149 = ScalarParameter() {defaultValue=3, group=TopSideProjection, parameterName=TS01_Mip_levels_01, sortPriority=2}
  n150 = TextureSample!(UVs=@n148, Tex=NamedRerouteUsage(), Level=@n149, Apply View MipBias) {mipValueMode=TMVM_MipLevel}
  n151 = TextureSample!(UVs=@n148, Tex=NamedRerouteUsage(), Apply View MipBias)
  n152 = NamedRerouteUsage()
  n153 = VectorParameter() {group=TopSideProjection, parameterName=TS_Fill_alternate_Axis_Color_Mask_01, sortPriority=21, defaultValue=<rgba 0 0 0 1>}
  n154 = AppendVector(A=@n153.RGB, B=@n153.RGB)
  n155 = NamedRerouteUsage()
  n156 = StaticBoolParameter() {group=TopSideProjection, parameterName=TS01-USE-Miplevels-VS-Alternative_Axies, sortPriority=2}
  n157 = TextureObjectParameter() {
   group=TopSideProjection
   parameterName=TS_Mask_01
   samplerType=SAMPLERTYPE_Masks
   sortPriority=20
   texture=&T_BlackMask
  }
  n158 = LinearInterpolate(
   A=ScalarParameter() {defaultValue=1, group=TopSideProjection, parameterName=TS_Blend_Contrast, sortPriority=-1}
   B=ScalarParameter() {defaultValue=2, group=TopSideProjection, parameterName=TS_Blend_Contrast_Mech, sortPriority=-1}
   Alpha=@n15
  )
  n159 = ScalarParameter() {group=TopSideProjection, parameterName=TS_Fill_alternate_Axis_Opacity 01, sortPriority=2}
  n160 = Max(A=@n159, B=ScalarParameter() {group=TopSideProjection, parameterName=TS_Fill_alternate_Axis_Opacity 01 VehicleSide, sortPriority=2})
  n161 = NamedRerouteUsage()
  n162 = LinearInterpolate(
   A=StaticSwitch(
    True=Constant() {r=1}
    False=LinearInterpolate(A=@n159, B=@n160, Alpha=@n161)
    Value=StaticBoolParameter() {group=TopSideProjection, parameterName=TS01-USE-Miplevels-VS-Alternative_Axies, sortPriority=2}
   )
   B=ScalarParameter() {group=TopSideProjection, parameterName=TS_Fill_alternate_Axis_Opacity_Character 01, sortPriority=2}
   Alpha=NamedRerouteUsage()
  )
  n163 = VectorParameter() {group=TopSideProjection, parameterName=TS_Fill_Z_Axis_Weapon_Color_Mask_01, sortPriority=21, defaultValue=<rgba 0 0 0 1>}
  n164 = VectorParameter() {group=TopSideProjection_UVS, parameterName=TS-Rotation-Weapon-01---------------------, sortPriority=1, defaultValue=<rgba 0 0 -0.25 1>}
  n165 = VectorParameter() {group=TopSideProjection_UVS, parameterName=TS_Rotation_Vehicle_01, sortPriority=2, defaultValue=<rgba 0 0 0 1>}
  n166 = VectorParameter() {
   desc=must activate TS_Vehicle_Side_01 UVs
   group=TopSideProjection_UVS
   parameterName=TS_Rotation_Vehicle_Side_01
   sortPriority=2
   defaultValue=<rgba 0 0 0 1>
  }
  n167 = MaterialFunctionCall(Float4=NamedRerouteUsage()) {materialFunction=&BreakOutFloat4Components}
  n168 = NamedRerouteDeclaration(None=Multiply(A=@n167.G, B=ScalarParameter() {group=TopSideProjection_UVS, parameterName="TS_Vehicle_Side_01 UVs 0=Off 1=Active"})) {name=TS_Vehicle_Side_01 ON, nodeColor=<rgba 1 0.5966 0 1>}
  n169 = VectorParameter() {group=TopSideProjection_UVS, parameterName=TS_Rotation_Mech_01, sortPriority=3, defaultValue=<rgba 0.25 0.25 0 1>}
  n170 = VectorParameter() {group=TopSideProjection_UVS, parameterName=TS_Rotation_Character_01, sortPriority=3, defaultValue=<rgba 0.25 0.25 0 1>}
  n171 = LinearInterpolate(
   A=LinearInterpolate(A=@n164.RGB, B=LinearInterpolate(A=@n165.RGB, B=@n166.RGB, Alpha=@n168), Alpha=@n167.R)
   B=LinearInterpolate(A=@n169.RGB, B=@n170.RGB, Alpha=@n167.A)
   Alpha=@n167.B
  )
  n172 = NamedRerouteUsage()
  n173 = MaterialFunctionCall(Float4=NamedRerouteUsage()) {materialFunction=&BreakOutFloat4Components}
  n174 = NamedRerouteUsage()
  n175 = VectorParameter() {group=TopSideProjection_UVS, parameterName=TS-WarpUVS-01-Anitmated Weapon, sortPriority=57, defaultValue=<rgba 0 0 0 1>}
  n176 = VectorParameter() {group=TopSideProjection_UVS, parameterName=TS-WarpUVS-01-Anitmated Vehicle, sortPriority=57, defaultValue=<rgba 0 0 0 1>}
  n177 = Multiply(A=Time() {period=1}, B=ComponentMask(None=LinearInterpolate(A=@n175.RGB, B=@n176.RGB, Alpha=@n14)) {g=true, r=true})
  n178 = AppendVector(A=@n177, B=Constant())
  n179 = ScalarParameter() {defaultValue=1, group=TopSideProjection_UVS, parameterName=TS-WarpUVS-01-TileRate Character, sortPriority=54}
  n180 = LinearInterpolate(
   A=ScalarParameter() {defaultValue=1, group=TopSideProjection_UVS, parameterName=TS-WarpUVS-01-TileRate Weapon, sortPriority=53}
   B=Multiply(
    A=ScalarParameter() {defaultValue=1, group=TopSideProjection_UVS, parameterName=TS-WarpUVS-01-TileRate Vehicle, sortPriority=54}
    B=LinearInterpolate(B=@n179, Alpha=@n173.A) {constA=1}
   )
   Alpha=@n14
  )
  n181 = Multiply(A=TextureCoordinate(), B=@n180)
  n182 = TextureSampleParameter2D(UVs=StaticSwitchParameter(
    True=MaterialFunctionCall(
     Caculated_Bounds=@n172
     Is Vehicle=@n173.R
     Is VehicleSide=@n173.R
     Is Mech=@n173.R
     UV_Rotation Bypass=@n174
     Uvs_Offset=@n178
     Uvs_TileRate=@n180
    ) {materialFunction=&MF_Projection_Side_Normal}.UVs
    False=Add(A=@n181, B=@n177)
   ) {group=TopSideProjection_UVS, parameterName=TS-WarpUVS-Use-SideProjection-01, sortPriority=50}) {
   group=TopSideProjection_UVS
   parameterName=TS-WarpUVS-01-Texture
   sortPriority=52
   texture=&AtmosphericCloudNoise01
  }
  n183 = Multiply(
   A=ChannelMaskParameter(None=@n182.RGBA) {group=TopSideProjection_UVS, parameterName=TS-WarpUVS-01_Channel, sortPriority=51}.R
   B=LinearInterpolate(
    A=ScalarParameter() {group=TopSideProjection_UVS, parameterName=TS-WarpUVS-01-Amount Weapon, sortPriority=55}
    B=ScalarParameter() {group=TopSideProjection_UVS, parameterName=TS-WarpUVS-01-Amount_Vehicle, sortPriority=56}
    Alpha=@n14
   )
  )
  n184 = StaticSwitchParameter(True=Add(A=@n41, B=@n183), False=@n41) {group=TopSideProjection_UVS, parameterName=TS-WarpUVS-01---------------------, sortPriority=50}
  n185 = StaticBoolParameter() {group=TopSideProjection, parameterName=TS-Use-Side-Projectiion-01, sortPriority=10}
  n186 = StaticSwitch(
   True=StaticSwitch(True=LinearInterpolate(A=@n150.RGBA, B=@n151.RGBA, Alpha=@n152), False=LinearInterpolate(A=@n154, B=@n151.RGBA, Alpha=@n155), Value=@n156)
   False=MaterialFunctionCall(
    Texture=NamedRerouteDeclaration(None=@n157) {name=Tex_TS01_Msk, nodeColor=<rgba 0.3125 0 1 1>}
    Rotation_Axis=@n171
    Caculated_Bounds=@n184
    Pre_Skinned_Normals=@n127
    Blend_Contrast=@n158
    Fill_alternate_Axis_Opacity=@n162
    "Fill_alternate_Axis_Color_A=Roughness"=@n154
    Is_Vehicle=@n14
    Is_Mech=@n15
    Enable_Fill_Z_Weapon=StaticBool() {value=true}
    Fill_Z_Weapon_opacity=ScalarParameter() {group=TopSideProjection, parameterName=TS_Fill_Z_Axis_Weapon_Opacity 01, sortPriority=2}
    Fill_Z_Weapon_Color=AppendVector(A=@n163.RGB, B=@n163.RGB)
   ) {materialFunction=&MF_TopsSideProjection_Base}.Combined
   Value=@n185
  )
  n187 = StaticSwitchParameter(
   True=ChannelMaskParameter(None=@n139) {group=TopSideProjection_Emissive, parameterName=TS_WarpEmissive01MaskChannel_02, sortPriority=37}.R
   False=ChannelMaskParameter(None=@n186) {group=TopSideProjection_Emissive, parameterName=TS_WarpEmissive01MaskChannel_01, sortPriority=37}.R
  ) {group=TopSideProjection_Emissive, parameterName=TS_WarpEmissive01_TS_Mask_02, sortPriority=36}
  n188 = ScalarParameter() {group=TopSideProjection_Emissive, parameterName=TS_WarpEmissive01_Warp_Offset_Range, sortPriority=38}
  n189 = StaticSwitchParameter(
   True=Multiply(
    A=Add(A=@n187, B=@n188)
    B=ScalarParameter() {group=TopSideProjection_Emissive, parameterName=TS_WarpEmissive01_Warp_Amount, sortPriority=38}
   )
   False=Constant()
  ) {group=TopSideProjection_Emissive, parameterName=TS_WarpEmissive01--TS_Mask_01-----------, sortPriority=35}
  n190 = ScalarParameter() {defaultValue=1, group=TopSideProjection_Emissive, parameterName=TS_Emissive01_Texture_Rotate_Vehicle_Side, sortPriority=13}
  n191 = NamedRerouteDeclaration(None=FunctionInput() {bUsePreviewValueAsDefault=true, inputName=Side Project Vehicle, inputType=FunctionInput_Scalar, sortPriority=100}) {name=IsVehicleSide, nodeColor=<rgba 0 1 0.0156 1>}
  n192 = MaterialFunctionCall!(
   ForwardAxis=@n76
   Bounds=@n17.Result
   Pre-Skinned Local Position
   isVehicle=@n14
   isMech=@n15
   Side Project Vehicle=@n191
   WeaponOffset=StaticSwitch(
    True=@n94.RGB
    False=ScalarParameter() {defaultValue=-0.5, group=TopSideProjection_Emissive, parameterName="TS_Emissive01_OffsetWeapon ", sortPriority=20}
    Value=@n95
   )
   WeaponFalloff=@n96
   VehicleOffset=StaticSwitch(
    True=@n97.RGB
    False=ScalarParameter() {defaultValue=-0.5, group=TopSideProjection_Emissive, parameterName="TS_Emissive01_OffsetVehicle ", sortPriority=21}
    Value=@n95
   )
   VehicleFalloff=ScalarParameter() {defaultValue=0.5, group=TopSideProjection_Emissive, parameterName="TS_Emissive01_FalloffVehicle ", sortPriority=24}
   MechOffset=StaticSwitch(True=LinearInterpolate(A=@n98.RGB, B=@n99.RGB, Alpha=@n100.A), False=LinearInterpolate(A=@n101, B=@n102, Alpha=@n100.A), Value=@n95)
   MechFalloff=ScalarParameter() {defaultValue=0.5, group=TopSideProjection_Emissive, parameterName="TS_Emissive01_FalloffMech ", sortPriority=25}
   Invert_Gradent=StaticBoolParameter() {group=TopSideProjection_Emissive, parameterName=TS_Emissive01_Invert_Gradent, sortPriority=11}
   UV_Spread_XY=LinearInterpolate(A=@n103.RGB, B=@n104.RGB, Alpha=@n100.B)
   UV_Animated_XY=LinearInterpolate(A=LinearInterpolate(A=@n105.RGB, B=@n106.RGB, Alpha=@n100.R), B=LinearInterpolate(A=@n107.RGB, B=@n108.RGB, Alpha=@n100.A), Alpha=@n100.B)
   Warp_UVs=@n189
   Rotate_Mech=ScalarParameter() {defaultValue=1, group=TopSideProjection_Emissive, parameterName=TS_Emissive01_Texture_Rotate_Mech, sortPriority=13}
   Rotate_Vehicle_Side=@n190
   Rotate_Vehicle=ScalarParameter() {defaultValue=1, group=TopSideProjection_Emissive, parameterName=TS_Emissive01_Texture_Rotate_Vehicle, sortPriority=13}
  ) {materialFunction=&/MF_Wrap_Gradent_Creation}
  n193 = TextureObjectParameter() {
   group=TopSideProjection_Emissive
   parameterName=Ts_Emissive01Gradent_Texture
   samplerType=SAMPLERTYPE_Masks
   sortPriority=12
   texture=/Game/Global/Textures/Default/Masks_and_Gradients/T_VertGradient_MASK
  }
  n194 = StaticSwitch(
   True=ChannelMaskParameter(None=TextureSample(UVs=@n192.Uvs, Tex=@n193).RGBA) {group=TopSideProjection_Emissive, parameterName=Ts_Emissive01_TextureChannel, sortPriority=11}.R
   False=@n192.Gradent_B
   Value=@n95
  )
  n195 = ScalarParameter() {defaultValue=-9999999, group=TopSideProjection_Emissive_Reactive, parameterName=TS_Emissive02_Time_BP_Gradent, sortPriority=212}
  n196 = Subtract(
   A=Time()
   B=Add(
    A=ScalarParameter() {group=TopSideProjection_Emissive_Reactive, parameterName=TS_Emissive02_Time_BP_Delay_Gradent, sortPriority=211}
    B=@n195
   )
  )
  n197 = ScalarParameter() {defaultValue=0.25, group=TopSideProjection_Emissive_Reactive, parameterName=TS_Emissive02_Time_BP_Duration_Gradent, sortPriority=211}
  n198 = ScalarParameter() {defaultValue=0.25, group=TopSideProjection_Emissive_Reactive, parameterName=TS_Emissive02_Time_BP_Falloff_Gradent, sortPriority=211}
  n199 = Power(Base=Saturate(None=Divide(A=@n196, B=@n197)), Exp=@n198)
  n200 = StaticBoolParameter() {group=TopSideProjection_Emissive_Reactive, parameterName=TS_Emissive02_Reactive_Gradent---------------------, sortPriority=200}
  n201 = MaterialFunctionCall(Color=Abs(None=@n127)) {materialFunction=&SplitComponents}
  n202 = MaterialFunctionCall(Float4=NamedRerouteUsage()) {materialFunction=&BreakOutFloat4Components}
  n203 = LinearInterpolate(A=@n201.B, B=@n201.B, Alpha=@n202.G)
  n204 = MaterialFunctionCall(In=LinearInterpolate(A=LinearInterpolate(A=@n201.R, B=@n203, Alpha=@n202.R), B=@n201.G, Alpha=@n202.B), Contrast=@n158) {materialFunction=&CheapContrast}
  n205 = Add(
   A=@n204.Result
   B=ScalarParameter() {defaultValue=1, group=TopSideProjection_Emissive, parameterName=TS_AEmissive_TriPlanner_Masking_01, sortPriority=3}
  ) {constB=0}
  n206 = ScalarParameter() {defaultValue=1, group=TopSideProjection, parameterName=TS_Blend_Contrast, sortPriority=-1}
  n207 = MaterialFunctionCall(In=Abs(None=ComponentMask(None=@n127) {b=true}), Contrast=@n206) {materialFunction=&CheapContrast}
  n208 = MaterialFunctionCall(In=Abs(None=ComponentMask(None=@n127) {g=true}), Contrast=@n206) {materialFunction=&CheapContrast}
  n209 = Saturate(None=Subtract(A=@n207.Result, B=@n208.Result))
  n210 = MaterialFunctionCall(In=Abs(None=ComponentMask(None=@n127) {r=true}), Contrast=@n206) {materialFunction=&CheapContrast}
  n211 = ScalarParameter() {defaultValue=1, group=TopSideProjection_Emissive, parameterName=TS_AEmissive_alternate_Axis_Masking_Weapon, sortPriority=3}
  n212 = Saturate(None=Add(A=Add(A=@n209, B=@n210.Result), B=@n211))
  n213 = Saturate(None=Subtract(A=@n207.Result, B=@n210.Result))
  n214 = ScalarParameter() {defaultValue=1, group=TopSideProjection_Emissive, parameterName=TS_AEmissive_alternate_Axis_Masking_Vehicle, sortPriority=3}
  n215 = Saturate(None=Add(A=Add(A=@n213, B=@n208.Result), B=@n214))
  n216 = Saturate(None=LinearInterpolate(A=LinearInterpolate(A=@n212, B=@n215, Alpha=@n14), Alpha=@n15))
  n217 = Multiply(
   A=StaticSwitch(True=LinearInterpolate(A=@n93, B=@n194, Alpha=@n199), False=@n194, Value=@n200)
   B=StaticSwitch(True=Saturate(None=@n205), False=@n216, Value=@n185)
  )
  n218 = StaticSwitchParameter(
   True=ChannelMaskParameter(None=@n139) {group=TopSideProjection_Emissive, parameterName=TS_Emissive01_MaskChannel_02, sortPriority=6}.R
   False=ChannelMaskParameter(None=@n186) {group=TopSideProjection_Emissive, parameterName=TS_Emissive01_MaskChannel_01, sortPriority=6}.R
  ) {group=TopSideProjection_Emissive, parameterName="TS-Emissive01--TS-Mask-01--->TS-Mask-02", sortPriority=5}
  n219 = Multiply(A=@n218, B=@n217)
  n220 = Multiply(A=@n219, B=Add(A=@n219, B=@n217))
  n221 = StaticSwitchParameter(
   True=MaterialFunctionCall(
    Transition Phase=@n217
    Height Texture=@n218
    Contrast=ScalarParameter() {group=TopSideProjection_Emissive, parameterName=TS_Emissive01_HeightLerp_Contrast, sortPriority=8}
   ) {materialFunction=&HeightLerp}.Alpha
   False=StaticSwitchParameter(True=Saturate(None=@n220), False=@n217) {group=TopSideProjection_Emissive, parameterName=TS_Emissive01_X+X, sortPriority=7}
  ) {group=TopSideProjection_Emissive, parameterName=TS_Emissive01_HeightLerp, sortPriority=7}
  n222 = Saturate(None=Multiply(
    A=StaticSwitchParameter(True=@n91, False=@n92) {group=TopSideProjection_Emissive_Reactive, parameterName=TS_Emissive01_Reactive_Brightness---------------------}
    B=@n221
   ))
  n223 = VectorParameter() {group=TopSideProjection, parameterName=TS_SimpleDifuse_Rougness_02, sortPriority=56, defaultValue=<rgba 0 0 0 1>}
  n224 = AppendVector(A=@n223.RGB, B=@n223.RGB)
  n225 = TextureObjectParameter() {group=TopSideProjection, parameterName=TS_BaseColor_02, sortPriority=56, texture=/Game/Weapons/WeaponSkins/Wraps/Textures/UVCheaker}
  n226 = VectorParameter() {group=TopSideProjection, parameterName="TS_Fill_alternate_Axis_Color_A=Roughness 02", sortPriority=52}
  n227 = AppendVector(A=@n226.RGB, B=@n226.RGB)
  n228 = NamedRerouteUsage()
  n229 = TextureSample!(UVs=@n131.UVs, Tex=@n228, Level=@n133, Apply View MipBias) {mipValueMode=TMVM_MipLevel}
  n230 = TextureSample!(UVs=@n131.UVs, Tex=@n228, Apply View MipBias)
  n231 = StaticSwitch(
   True=StaticSwitch(True=LinearInterpolate(A=@n229.RGBA, B=@n230.RGBA, Alpha=@n136), False=LinearInterpolate(A=@n227, B=@n230.RGBA, Alpha=@n137), Value=@n138)
   False=MaterialFunctionCall(
    Texture=NamedRerouteDeclaration(None=@n225) {name=TS_BaseColor_02, nodeColor=<rgba 1 0 0.5391 1>}
    Rotation_Axis=@n125
    Caculated_Bounds=@n126
    Pre_Skinned_Normals=@n127
    Blend_Contrast=@n110
    Fill_alternate_Axis_Opacity=@n114
    "Fill_alternate_Axis_Color_A=Roughness"=@n227
    Is_Vehicle=@n14
    Is_Mech=@n15
   ) {materialFunction=&MF_TopsSideProjection_Base}.Combined
   Value=@n128
  )
  n232 = VectorParameter() {group=TopSideProjection, parameterName=TS_SimpleDifuse_Rougness_01, sortPriority=11}
  n233 = TextureSample(UVs=@n148, Tex=NamedRerouteUsage())
  n234 = LinearInterpolate(
   A=TextureSample!(UVs=@n148, Tex=NamedRerouteUsage(), Level=@n149, Apply View MipBias) {mipValueMode=TMVM_MipLevel}.RGBA
   B=@n233.RGBA
   Alpha=@n152
  )
  n235 = VectorParameter() {group=TopSideProjection, parameterName="TS_Fill_alternate_Axis_Color_A=Roughness 01", sortPriority=3}
  n236 = AppendVector(A=@n235.RGB, B=@n235.RGB)
  n237 = LinearInterpolate(A=@n236, B=@n233.RGBA, Alpha=@n155)
  n238 = TextureObjectParameter() {group=TopSideProjection, parameterName=TS_BaseColor_01, sortPriority=11, texture=&UVCheaker}
  n239 = NamedRerouteDeclaration(None=@n238) {name=Tex_TS01_BC, nodeColor=<rgba 0.3125 0 1 1>}
  n240 = StaticSwitchParameter(
   True=AppendVector(A=@n232.RGB, B=@n232.RGB)
   False=StaticSwitch(
    True=StaticSwitch(True=@n234, False=@n237, Value=@n156)
    False=MaterialFunctionCall(
     Texture=@n239
     Rotation_Axis=@n171
     Caculated_Bounds=@n184
     Pre_Skinned_Normals=@n127
     Blend_Contrast=@n158
     Fill_alternate_Axis_Opacity=@n162
     "Fill_alternate_Axis_Color_A=Roughness"=@n236
     Is_Vehicle=@n14
     Is_Mech=@n15
    ) {materialFunction=&MF_TopsSideProjection_Base}.Combined
    Value=@n185
   )
  ) {group=TopSideProjection, parameterName=TS-SimpleDifuse-Rougness-01----------------------, sortPriority=10}
  n241 = MaterialFunctionCall(Float4=NamedRerouteUsage()) {materialFunction=&BreakOutFloat4Components}
  n242 = VectorParameter() {group=TopSideProjection, parameterName=TS_Blend_OffsetWeapon, sortPriority=85, defaultValue=<rgba 0 -1 0 0>}
  n243 = StaticBoolParameter() {group=TopSideProjection, parameterName="Ts-Gradent-->Texture", sortPriority=73}
  n244 = VectorParameter() {group=TopSideProjection, parameterName=TS_Blend_OffsetVehicle, sortPriority=86, defaultValue=<rgba -1 0 0 0>}
  n245 = VectorParameter() {group=TopSideProjection, parameterName=TS_Blend_OffsetVehicleSide, sortPriority=86, defaultValue=<rgba -1 0 0 0>}
  n246 = Multiply(A=@n191, B=ScalarParameter() {group=TopSideProjection, parameterName="TS_Blend__OffsetVehicleSide 0=Off 1=Active", sortPriority=86})
  n247 = ScalarParameter() {defaultValue=-1, group=TopSideProjection, parameterName="TS_Blend_OffsetVehicle ", sortPriority=86}
  n248 = ScalarParameter() {defaultValue=-1, group=TopSideProjection, parameterName="TS_Blend_OffsetVehicleSide ", sortPriority=86}
  n249 = VectorParameter() {group=TopSideProjection, parameterName=TS_Blend_OffsetMech, sortPriority=87, defaultValue=<rgba 0 0 -1 0>}
  n250 = VectorParameter() {group=TopSideProjection, parameterName="TS_Blend_OffsetCharacter ", sortPriority=87, defaultValue=<rgba 0 0 -1 0>}
  n251 = ScalarParameter() {defaultValue=-1, group=TopSideProjection, parameterName="TS_Blend_OffsetMech ", sortPriority=87}
  n252 = ScalarParameter() {defaultValue=-1, group=TopSideProjection, parameterName=TS_Blend_OffsetCharacter, sortPriority=87}
  n253 = VectorParameter() {group=TopSideProjection, parameterName=TS_Blend_Spread_XY, sortPriority=80, defaultValue=<rgba 1 1 1 0>}
  n254 = StaticSwitchParameter(
   True=ChannelMaskParameter(None=@n139) {group=TopSideProjection, parameterName=TS_BlendWarpMaskChannel_02, sortPriority=93}.R
   False=ChannelMaskParameter(None=@n186) {group=TopSideProjection, parameterName=TS_BlendWarpMaskChannel_01, sortPriority=93}.R
  ) {group=TopSideProjection, parameterName=TS-BlendWarp-TS_Mask_02, sortPriority=91}
  n255 = Constant()
  n256 = MaterialFunctionCall(
   ForwardAxis=@n76
   Bounds=@n17.Result
   isVehicle=@n241.R
   isMech=@n241.R
   Side Project Vehicle=@n191
   WeaponOffset=StaticSwitch(
    True=@n242.RGB
    False=ScalarParameter() {defaultValue=-1, group=TopSideProjection, parameterName="TS_Blend_OffsetWeapon ", sortPriority=85}
    Value=@n243
   )
   WeaponFalloff=ScalarParameter() {defaultValue=-0.5, group=TopSideProjection, parameterName="TS_Blend_FalloffWeapon (negative inverts)", sortPriority=88}
   VehicleOffset=StaticSwitch(True=LinearInterpolate(A=@n244.RGB, B=@n245.RGB, Alpha=@n246), False=LinearInterpolate(A=@n247, B=@n248, Alpha=@n246), Value=@n243)
   VehicleFalloff=ScalarParameter() {defaultValue=-0.5, group=TopSideProjection, parameterName=TS_Blend_FalloffVehicle, sortPriority=89}
   MechOffset=StaticSwitch(True=LinearInterpolate(A=@n249.RGB, B=@n250.RGB, Alpha=@n241.A), False=LinearInterpolate(A=@n251, B=@n252, Alpha=@n241.A), Value=@n243)
   MechFalloff=LinearInterpolate(
    A=ScalarParameter() {defaultValue=-0.5, group=TopSideProjection, parameterName=TS_Blend_FalloffMech, sortPriority=90}
    B=ScalarParameter() {defaultValue=-0.5, group=TopSideProjection, parameterName=TS_Blend_FalloffCharacter, sortPriority=90}
    Alpha=@n241.A
   )
   Invert_Gradent=StaticBoolParameter() {group=TopSideProjection, parameterName=TS_Blend_Invert_Gradent, sortPriority=81}
   UV_Spread_XY=@n253.RGB
   UV_Animated_XY=VectorParameter() {group=TopSideProjection, parameterName=TS_Blend_UV_Animated_XY, sortPriority=82}.RGB
   Warp_UVs=Multiply(
    A=StaticSwitchParameter(True=@n254, False=@n255) {group=TopSideProjection, parameterName=TS-BlendWarp--TS_Mask_01-----------, sortPriority=90}
    B=ScalarParameter() {group=TopSideProjection, parameterName=TS_BlendWarp_Amount, sortPriority=92}
   )
   Rotate_Mech=ScalarParameter() {defaultValue=1, group=TopSideProjection, parameterName=TS_Gradent_Texture_Rotate_Mech, sortPriority=74}
   Rotate_Vehicle_Side=ScalarParameter() {defaultValue=1, group=TopSideProjection, parameterName=TS_Gradent_Texture_Rotate_Vehicle_Side, sortPriority=74}
   Rotate_Vehicle=ScalarParameter() {defaultValue=1, group=TopSideProjection, parameterName=TS_Gradent_Texture_Rotate_Vehicle, sortPriority=74}
  ) {materialFunction=&MF_Wrap_Gradent_Creation}
  n257 = TextureObjectParameter() {
   group=TopSideProjection
   parameterName=Ts_BlendGradent_Texture
   samplerType=SAMPLERTYPE_Masks
   sortPriority=75
   texture=&T_VertGradient_MASK
  }
  n258 = ScalarParameter() {defaultValue=1, group=TopSideProjection, parameterName=TS_Blend_Power, sortPriority=84}
  n259 = StaticSwitch(
   True=ChannelMaskParameter(None=TextureSample!(UVs=@n256.Uvs, Tex=@n257, Apply View MipBias) {samplerType=SAMPLERTYPE_Masks, texture=&T_VertGradient_MASK}.RGBA) {group=TopSideProjection, parameterName=Ts_Blend_TextureChannel, sortPriority=74}.R
   False=Saturate(None=Power(Base=@n256.Gradent_B, Exp=@n258))
   Value=@n243
  )
  n260 = StaticSwitchParameter(
   True=ChannelMaskParameter(None=@n139) {group=TopSideProjection, parameterName=TS_Blend_HieghtlerpMaskChannel_02, sortPriority=72}.R
   False=ChannelMaskParameter(None=@n186) {group=TopSideProjection, parameterName=TS_Blend_HieghtlerpMaskChannel_01, sortPriority=72}.R
  ) {group=TopSideProjection, parameterName="TS-Blend-Hieghtlerp--TS-Mask-01--->TS-Mask-02", sortPriority=71}
  n261 = Multiply(A=@n260, B=@n259)
  n262 = Multiply(A=@n261, B=Add(A=@n261, B=@n259))
  n263 = StaticSwitchParameter(
   True=MaterialFunctionCall(
    Transition Phase=@n259
    Height Texture=@n260
    Contrast=ScalarParameter() {defaultValue=1, group=TopSideProjection, parameterName=TS_Blend_Hieghtlerp_Contrast, sortPriority=71}
   ) {materialFunction=&HeightLerp}.Alpha
   False=StaticSwitchParameter(True=Saturate(None=@n262), False=@n259) {group=TopSideProjection, parameterName="Ts-Blend-Gradent-->X+X", sortPriority=70}
  ) {group=TopSideProjection, parameterName="Ts-Blend-Gradent-->Height_Lerp", sortPriority=70}
  n264 = Saturate(None=@n263)
  n265 = ScalarParameter() {defaultValue=-0.5, group=TopSideProjection, parameterName="TS_Blend_Transition_OffsetWeapon (Only02)", sortPriority=301}
  n266 = ScalarParameter() {defaultValue=-0.5, group=TopSideProjection, parameterName="TS_Blend_Transition_OffsetVehicle (Only02)", sortPriority=302}
  n267 = ScalarParameter() {defaultValue=0.1, group=TopSideProjection, parameterName="TS_Blend_Transition_FalloffWeapon (Only02) negative inverts", sortPriority=305}
  n268 = ScalarParameter() {defaultValue=0.1, group=TopSideProjection, parameterName="TS_Blend_Transition_FalloffVehicle (Only02)", sortPriority=306}
  n269 = MaterialFunctionCall(
   Bounds=@n17.Result
   ForwardAxis=@n76
   Offset=LinearInterpolate(
    A=LinearInterpolate(A=@n265, B=@n266, Alpha=@n73.R)
    B=ScalarParameter() {defaultValue=-0.5, group=TopSideProjection, parameterName="TS_Blend_Transition_OffsetMehicle (Only02)", sortPriority=303}
    Alpha=@n73.B
   )
   FallOff=LinearInterpolate(
    A=LinearInterpolate(A=@n267, B=@n268, Alpha=@n73.R)
    B=ScalarParameter() {defaultValue=0.1, group=TopSideProjection, parameterName="TS_Blend_Transition_FalloffMech (Only02)", sortPriority=307}
    Alpha=@n73.B
   )
  ) {materialFunction=&MF_Wrap_Gradent_Simple}
  n270 = LinearInterpolate(
   A=ChannelMaskParameter(None=@n139) {group=TopSideProjection, parameterName=TS_Blend_Transition_MaskChannel_02, sortPriority=300}.R
   Alpha=@n269.Clamped_Data
  )
  n271 = StaticSwitch(
   True=Saturate(None=Multiply(A=@n264, B=@n270))
   False=@n263
   Value=StaticBoolParameter() {group=TopSideProjection, parameterName=TS-Blend-UV/Transition-Split-02------------------, sortPriority=299}
  )
  n272 = StaticBoolParameter() {group=TopSideProjection, parameterName=TS-Blend-TS_01--TS_02------------------------------------------------, sortPriority=50}
  n273 = StaticBoolParameter() {group=TopSideProjection, parameterName=TS_UseMask from TS-Blend-TS_01--TS_02, sortPriority=-9}
  n274 = StaticSwitch(True=StaticBool(), False=@n272, Value=@n273)
  n275 = StaticSwitch(
   True=LinearInterpolate(
    A=@n240
    B=StaticSwitchParameter(True=@n224, False=@n231) {group=TopSideProjection, parameterName=TS-SimpleDifuse-Rougness-02---------------------, sortPriority=55}
    Alpha=@n271
   )
   False=@n240
   Value=@n274
  )
  n276 = ComponentMask(None=@n275) {b=true, g=true, r=true}
  n277 = StaticSwitchParameter(True=Multiply(A=@n276, B=OneMinus(None=@n222)), False=@n276) {group=TopSideProjection_Emissive, parameterName=TS-Emissive-01------------------------------------------------------, sortPriority=1}
  n278 = ScalarParameter() {defaultValue=1, group=TopSideProjection_Emissive_Reactive, parameterName=TS_Emissive02_Time_Gradent_Start, sortPriority=210}
  n279 = ScalarParameter() {defaultValue=1, group=TopSideProjection_Emissive_Reactive, parameterName=TS_Emissive02_Time_Grad_Offset_Start, sortPriority=301}
  n280 = ScalarParameter() {defaultValue=1, group=TopSideProjection_Emissive_Reactive, parameterName=TS_Emissive02_Time_Grad_Offset_End, sortPriority=301}
  n281 = ScalarParameter() {defaultValue=-9999999, group=TopSideProjection_Emissive_Reactive, parameterName=TS_Emissive02_Time_BP_Grad_Offset, sortPriority=302}
  n282 = ScalarParameter() {group=TopSideProjection_Emissive_Reactive, parameterName=TS_Emissive02_Time_BP_Delay_Grad_Offset, sortPriority=302}
  n283 = ScalarParameter() {defaultValue=0.25, group=TopSideProjection_Emissive_Reactive, parameterName=TS_Emissive02_Time_BP_Duration_Grad_Offset, sortPriority=302}
  n284 = Divide(A=Subtract(A=Time(), B=Add(A=@n281, B=@n282)), B=@n283)
  n285 = ScalarParameter() {defaultValue=0.25, group=TopSideProjection_Emissive_Reactive, parameterName=TS_Emissive02_Time_BP_Falloff_Grad_Offset, sortPriority=302}
  n286 = LinearInterpolate(A=@n279, B=@n280, Alpha=Power(Base=Saturate(None=@n284), Exp=@n285))
  n287 = Constant()
  n288 = VectorParameter() {group=TopSideProjection_Emissive, parameterName=TS_Emissive02_OffsetWeapon, sortPriority=124, defaultValue=<rgba 0 -1 0 0>}
  n289 = StaticBoolParameter() {group=TopSideProjection_Emissive, parameterName="TS_Emissive02_Gradent-->Texture", sortPriority=109}
  n290 = ScalarParameter() {defaultValue=0.5, group=TopSideProjection_Emissive, parameterName="TS_Emissive02_FalloffWeapon (negative inverts)", sortPriority=123}
  n291 = VectorParameter() {group=TopSideProjection_Emissive, parameterName=TS_Emissive02_OffsetVehicle, sortPriority=125, defaultValue=<rgba -1 0 0 0>}
  n292 = VectorParameter() {group=TopSideProjection_Emissive, parameterName=TS_Emissive02_OffsetMech, sortPriority=126, defaultValue=<rgba 0 0 -1 0>}
  n293 = VectorParameter() {group=TopSideProjection_Emissive, parameterName=TS_Emissive02_OffsetCharacter, sortPriority=126, defaultValue=<rgba 0 0 -1 0>}
  n294 = MaterialFunctionCall(Float4=NamedRerouteUsage()) {materialFunction=&BreakOutFloat4Components}
  n295 = ScalarParameter() {defaultValue=-0.5, group=TopSideProjection_Emissive, parameterName="TS_Emissive02_OffsetMech ", sortPriority=122}
  n296 = ScalarParameter() {defaultValue=-0.5, group=TopSideProjection_Emissive, parameterName=TS_Emissive02_OffsetCharacter_, sortPriority=122}
  n297 = VectorParameter() {group=TopSideProjection_Emissive, parameterName=TS_Emissive02_Spread_XY, sortPriority=120, defaultValue=<rgba 1 1 1 0>}
  n298 = VectorParameter() {group=TopSideProjection_Emissive, parameterName=TS_Emissive02_Spread_Mech_XY, sortPriority=119, defaultValue=<rgba 1 1 1 0>}
  n299 = VectorParameter() {group=TopSideProjection_Emissive, parameterName=TS_Emissive02_Animated_WeaponXY, sortPriority=130}
  n300 = VectorParameter() {group=TopSideProjection_Emissive, parameterName=TS_Emissive02_Animated_VehicleXY, sortPriority=131}
  n301 = VectorParameter() {group=TopSideProjection_Emissive, parameterName=TS_Emissive02_Animated_MechXY, sortPriority=132}
  n302 = VectorParameter() {group=TopSideProjection_Emissive, parameterName=TS_Emissive02_Animated_CharacterXY, sortPriority=132}
  n303 = StaticSwitchParameter(
   True=ChannelMaskParameter(None=@n139) {group=TopSideProjection_Emissive, parameterName=TS_WarpEmissive02MaskChannel_02, sortPriority=137}.R
   False=ChannelMaskParameter(None=@n186) {group=TopSideProjection_Emissive, parameterName=TS_WarpEmissive02MaskChannel_01, sortPriority=137}.R
  ) {group=TopSideProjection_Emissive, parameterName=TS_WarpEmissive02_TS_Mask_02, sortPriority=136}
  n304 = ScalarParameter() {group=TopSideProjection_Emissive, parameterName=TS_WarpEmissive02_Warp_Offset_Range, sortPriority=138}
  n305 = StaticSwitchParameter(
   True=Multiply(
    A=Add(A=@n303, B=@n304)
    B=ScalarParameter() {group=TopSideProjection_Emissive, parameterName=TS_WarpEmissive02_Warp_Amount, sortPriority=138}
   )
   False=Constant()
  ) {group=TopSideProjection_Emissive, parameterName=TS_WarpEmissive02--TS_Mask_01-----------, sortPriority=135}
  n306 = ScalarParameter() {defaultValue=1, group=TopSideProjection_Emissive, parameterName=TS_Emissive02_Texture_Rotate_Vehicle_Side, sortPriority=113}
  n307 = MaterialFunctionCall(
   ForwardAxis=@n76
   Bounds=@n17.Result
   isVehicle=@n14
   isMech=@n15
   Side Project Vehicle=@n191
   WeaponOffset=StaticSwitch(
    True=@n288.RGB
    False=ScalarParameter() {defaultValue=-0.5, group=TopSideProjection_Emissive, parameterName="TS_Emissive02_OffsetWeapon ", sortPriority=120}
    Value=@n289
   )
   WeaponFalloff=@n290
   VehicleOffset=StaticSwitch(
    True=@n291.RGB
    False=ScalarParameter() {defaultValue=-0.5, group=TopSideProjection_Emissive, parameterName="TS_Emissive02_OffsetVehicle ", sortPriority=121}
    Value=@n289
   )
   VehicleFalloff=ScalarParameter() {defaultValue=0.5, group=TopSideProjection_Emissive, parameterName="TS_Emissive02_FalloffVehicle ", sortPriority=124}
   MechOffset=StaticSwitch(True=LinearInterpolate(A=@n292.RGB, B=@n293.RGB, Alpha=@n294.A), False=LinearInterpolate(A=@n295, B=@n296, Alpha=@n294.A), Value=@n289)
   MechFalloff=ScalarParameter() {defaultValue=0.5, group=TopSideProjection_Emissive, parameterName=TS_Emissive02_FalloffMech, sortPriority=125}
   Invert_Gradent=StaticBoolParameter() {group=TopSideProjection_Emissive_Reactive, parameterName=TS_Emissive02_Invert_Gradent, sortPriority=110}
   UV_Spread_XY=LinearInterpolate(A=@n297.RGB, B=@n298.RGB, Alpha=@n294.B)
   UV_Animated_XY=LinearInterpolate(A=LinearInterpolate(A=@n299.RGB, B=@n300.RGB, Alpha=@n294.R), B=LinearInterpolate(A=@n301.RGB, B=@n302.RGB, Alpha=@n294.A), Alpha=@n294.B)
   Warp_UVs=@n305
   Rotate_Mech=ScalarParameter() {defaultValue=1, group=TopSideProjection_Emissive, parameterName=TS_Emissive02_Texture_Rotate_Mech, sortPriority=113}
   Rotate_Vehicle_Side=@n306
   Rotate_Vehicle=ScalarParameter() {defaultValue=1, group=TopSideProjection_Emissive, parameterName=TS_Emissive02_Texture_Rotate_Vehicle, sortPriority=113}
  ) {materialFunction=&MF_Wrap_Gradent_Creation}
  n308 = StaticBoolParameter() {group=TopSideProjection_Emissive_Reactive, parameterName=TS_Emissive02_Reactive_Grad_Offset---------------------, sortPriority=300}
  n309 = StaticSwitch(True=Add(A=AppendVector(A=@n286, B=@n287), B=@n307.Uvs), False=@n307.Uvs, Value=@n308)
  n310 = TextureObjectParameter() {
   group=TopSideProjection_Emissive
   parameterName=TS_Emissive02_Gradent_Texture
   samplerType=SAMPLERTYPE_Masks
   sortPriority=111
   texture=&T_VertGradient_MASK
  }
  n311 = StaticSwitch(
   True=ChannelMaskParameter(None=TextureSample(UVs=@n309, Tex=@n310) {samplerType=SAMPLERTYPE_Masks}.RGBA) {group=TopSideProjection_Emissive, parameterName=TS_Emissive02_TextureChannel, sortPriority=110}.R
   False=@n307.Gradent_B
   Value=@n289
  )
  n312 = MaterialFunctionCall(Color=Abs(None=@n127)) {materialFunction=&SplitComponents}
  n313 = MaterialFunctionCall(Float4=NamedRerouteUsage()) {materialFunction=&BreakOutFloat4Components}
  n314 = LinearInterpolate(A=@n312.B, B=@n312.B, Alpha=@n313.G)
  n315 = MaterialFunctionCall(In=LinearInterpolate(A=LinearInterpolate(A=@n312.R, B=@n314, Alpha=@n313.R), B=@n312.G, Alpha=@n313.B), Contrast=@n110) {materialFunction=&CheapContrast}
  n316 = Add(
   A=ScalarParameter() {defaultValue=1, group=TopSideProjection_Emissive, parameterName=TS_AEmissive_TriPlanner_Masking_02, sortPriority=3}
   B=@n315.Result
  )
  n317 = Multiply(
   A=StaticSwitch(True=Saturate(None=@n316), False=@n216, Value=@n128)
   B=StaticSwitch(True=LinearInterpolate(A=@n278, B=@n311, Alpha=@n199), False=@n311, Value=@n200)
  )
  n318 = StaticSwitchParameter(
   True=ChannelMaskParameter(None=@n139) {group=TopSideProjection_Emissive, parameterName=TS_Emissive02_MaskChannel_02, sortPriority=105}.R
   False=ChannelMaskParameter(None=@n186) {group=TopSideProjection_Emissive, parameterName=TS_Emissive02_MaskChannel_01, sortPriority=105}.R
  ) {group=TopSideProjection_Emissive, parameterName="TS_Emissive02--TS-Mask-01--->TS-Mask-02", sortPriority=104}
  n319 = Multiply(A=@n318, B=@n317)
  n320 = Multiply(A=@n319, B=Add(A=@n319, B=@n317))
  n321 = StaticSwitchParameter(
   True=MaterialFunctionCall(
    Transition Phase=@n317
    Height Texture=@n318
    Contrast=ScalarParameter() {group=TopSideProjection_Emissive, parameterName=TS_Emissive02_HeightLerp_Contrast, sortPriority=108}
   ) {materialFunction=&HeightLerp}.Alpha
   False=StaticSwitchParameter(True=Saturate(None=@n320), False=@n317) {group=TopSideProjection_Emissive, parameterName=TS_Emissive02_X+X, sortPriority=107}
  ) {group=TopSideProjection_Emissive, parameterName=TS_Emissive02_HeightLerp, sortPriority=107}
  n322 = Subtract(
   A=Time()
   B=Add(
    A=ScalarParameter() {group=TopSideProjection_Emissive_Reactive, parameterName=TS_Emissive02_Time_BP_Delay_Brightness}
    B=ScalarParameter() {defaultValue=-9999999, group=TopSideProjection_Emissive_Reactive, parameterName=TS_Emissive02_Time_BP_Brightness}
   )
  )
  n323 = ScalarParameter() {defaultValue=0.25, group=TopSideProjection_Emissive_Reactive, parameterName=TS_Emissive02_Time_BP_Duration_Brightness}
  n324 = Power(
   Base=Saturate(None=Divide(A=@n322, B=@n323))
   Exp=ScalarParameter() {defaultValue=0.25, group=TopSideProjection_Emissive_Reactive, parameterName=TS_Emissive02_Time_BP_Falloff}
  )
  n325 = StaticSwitchParameter(
   True=LinearInterpolate(
    A=ScalarParameter() {group=TopSideProjection_Emissive_Reactive, parameterName=TS_Emissive02_Time_Brightness_DarkenBase_Start}
    B=ScalarParameter() {group=TopSideProjection_Emissive_Reactive, parameterName=TS_Emissive02_Time_Brightness_DarkenBase_End}
    Alpha=@n324
   )
   False=ScalarParameter() {group=TopSideProjection_Emissive, parameterName=TS-Emissive02-DarkenBase-------, sortPriority=103}
  ) {group=TopSideProjection_Emissive_Reactive, parameterName=TS_Emissive02_Reactive_Brightness---------------------}
  n326 = OneMinus(None=Saturate(None=Multiply(A=@n321, B=@n325)))
  n327 = StaticSwitchParameter(True=Multiply(A=@n277, B=@n326), False=@n277) {group=TopSideProjection_Emissive, parameterName=TS-Emissive-02------------------------------------------------------, sortPriority=100}
  n328 = ScalarParameter() {group=TopSideProjection, parameterName="TS_BaseColor----->Emissive", sortPriority=-10}
  n329 = LinearInterpolate(A=@n327, Alpha=@n328) {constB=0}
  n330 = NamedRerouteUsage()
  n331 = ScalarParameter() {group=TopSideProjection_Emissive, parameterName="TS_Emissive_01----->BaseColor", sortPriority=-10}
  n332 = NamedRerouteUsage()
  n333 = ScalarParameter() {group=TopSideProjection_Emissive, parameterName="TS_Emissive_02----->BaseColor", sortPriority=-10}
  n334 = Add(A=@n329, B=Add(A=LinearInterpolate(B=@n330, Alpha=@n331), B=LinearInterpolate(B=@n332, Alpha=@n333)))
  n335 = Constant3Vector()
  n336 = LinearInterpolate(A=@n335, B=@n131.Raw_Normal, Alpha=NamedRerouteUsage())
  n337 = MaterialFunctionCall!(
   Texture_Normals=@n129
   Rotation_Axis=@n125
   Caculated_Bounds=@n126
   Pre_Skinned_Normals=@n127
   Blend_Contrast=@n110
   Fill_alternate_Axis_Opacity=@n114
   Is_Vehicle=@n14
   Is_Mech=@n15
  ) {materialFunction=&/MF_TopsSideProjection_Normals}
  n338 = Constant3Vector()
  n339 = StaticSwitch(True=@n144.Raw_Normal, False=@n146.Raw_Normal, Value=@n147)
  n340 = NamedRerouteUsage()
  n341 = StaticSwitchParameter(
   True=Constant3Vector()
   False=StaticSwitch(
    True=LinearInterpolate(A=@n338, B=@n339, Alpha=@n155)
    False=MaterialFunctionCall(
     Texture_Normals=@n340
     Rotation_Axis=@n171
     Caculated_Bounds=@n184
     Pre_Skinned_Normals=@n127
     Blend_Contrast=@n158
     Fill_alternate_Axis_Opacity=@n162
     Is_Vehicle=@n14
     Is_Mech=@n15
    ) {materialFunction=&MF_TopsSideProjection_Normals}.Add_Surface_Normal
    Value=@n185
   )
  ) {defaultValue=true, group=TopSideProjection, parameterName=TS-SimpleNormal-01----------------------, sortPriority=15}
  n342 = LinearInterpolate(
   A=@n341
   B=StaticSwitchParameter(True=@n335, False=StaticSwitch(True=@n336, False=@n337.Add_Surface_Normal, Value=@n128)) {defaultValue=true, group=TopSideProjection, parameterName=TS-SimpleNormal-02----------------------, sortPriority=60}
   Alpha=@n271
  )
  n343 = LinearInterpolate(
   A=@n2.Normal
   B=Add(A=@n2.Normal, B=StaticSwitch(True=@n342, False=@n341, Value=@n274))
   Alpha=ScalarParameter() {defaultValue=1, group=TopSideProjection, parameterName=TS_Normal_Intensty, sortPriority=1}
  )
  n344 = Add(
   A=LinearInterpolate(B=@n327, Alpha=@n328)
   B=Add(A=LinearInterpolate(A=@n330, Alpha=@n331) {constB=0}, B=LinearInterpolate(A=@n332, Alpha=@n333) {constB=0})
  )
  n345 = StaticSwitchParameter(
   True=ChannelMaskParameter(None=@n139) {group=TopSideProjection, parameterName=TS_MetalicChannel_Mask_02, sortPriority=66}.R
   False=ScalarParameter() {group=TopSideProjection, parameterName=TS_Metalic_02, sortPriority=66}
  ) {group=TopSideProjection, parameterName=TS-MetalicMask_02, sortPriority=65}
  n346 = StaticSwitchParameter(
   True=ChannelMaskParameter(None=@n186) {group=TopSideProjection, parameterName=TS_MetalicChannel_Mask_01, sortPriority=26}.R
   False=ScalarParameter() {group=TopSideProjection, parameterName=TS_Metalic_01, sortPriority=26}
  ) {group=TopSideProjection, parameterName=TS-MetalicMask_01, sortPriority=25}
  n347 = ScalarParameter() {defaultValue=0.5, group=TopSideProjection, parameterName=TS_Specularty_Min, sortPriority=12}
  n348 = ScalarParameter() {defaultValue=0.5, group=TopSideProjection, parameterName=TS_Specularty_Max, sortPriority=12}
  n349 = ChannelMaskParameter(None=@n186) {group=TopSideProjection, parameterName=TS_SpecChannel_Mask_01, sortPriority=26}
  n350 = StaticSwitch(
   True=LinearInterpolate(
    A=@n349.R
    B=ChannelMaskParameter(None=@n139) {group=TopSideProjection, parameterName=TS_SpecChannel_Mask_02, sortPriority=66}.R
    Alpha=@n271
   )
   False=@n349.R
   Value=@n274
  )
  n351 = SetMaterialAttributes!(
   MaterialAttributes=@n2.MaterialAttributes
   BaseColor=@n334
   Roughness=LinearInterpolate(
    A=ScalarParameter() {group=TopSideProjection, parameterName=TS_Roughness_Min, sortPriority=13}
    B=ScalarParameter() {defaultValue=1, group=TopSideProjection, parameterName=TS_Roughness_Max, sortPriority=14}
    Alpha=ComponentMask(None=@n275) {a=true}
   )
   Normal=ShadingPathSwitch(Default=@n343, Deferred=@n343, Mobile=QualitySwitch(Default=@n2.Normal, Medium=@n343)) {desc=Migrated from FeatureLevelSwitch}
   Specular=StaticSwitchParameter(
    True=LinearInterpolate(A=@n347, B=@n348, Alpha=@n350)
    False=ScalarParameter() {defaultValue=0.5, group=TopSideProjection, parameterName=TS_Specularty, sortPriority=12}
   ) {group=TopSideProjection, parameterName=TS_Specularty_use_Channel_Mask, sortPriority=12}
   Metallic=StaticSwitch(True=LinearInterpolate(A=@n346, B=@n345, Alpha=@n271), False=@n346, Value=@n274)
   EmissiveColor=@n344
  ) {attributeSetTypes [
    #1
    #2
    #3
    #4
    #5
    #6
   ]}
  n352 = FunctionInput() {bUsePreviewValueAsDefault=true, inputName=Customization Mask A, inputType=FunctionInput_Scalar, sortPriority=24}
  n353 = NamedRerouteUsage()
  n354 = LinearInterpolate(
   A=Multiply(
    A=FunctionInput() {bUsePreviewValueAsDefault=true, inputName=Customization Mask B, inputType=FunctionInput_Scalar, sortPriority=24}
    B=ScalarParameter() {group=TopSideProjection, parameterName="TS_Accents_Mask  0=Off ", sortPriority=-10}
   )
   Alpha=@n353
  ) {constB=0}
  n355 = ScalarParameter() {group=TopSideProjection, parameterName=TS_Remove_Customization Mask, sortPriority=-10}
  n356 = Saturate(None=Add(A=Add(A=@n352, B=@n354), B=LinearInterpolate(A=@n355, Alpha=@n353) {constAlpha=0, constB=0}))
  n357 = StaticSwitch(True=StaticSwitch(True=Multiply(A=@n356, B=@n271), False=@n356, Value=@n273), False=@n356, Value=@n272)
  FunctionOutput(None=StaticSwitchParameter(
    True=BlendMaterialAttributes(A=@n2.MaterialAttributes, B=@n351, Alpha=@n357) {vertexAttributeBlendType=UseA}
    False=FunctionInput() {inputName=Material, inputType=FunctionInput_MaterialAttributes}
   ) {group=Customization, parameterName=TopSideProjection})
  FunctionOutput(None=@n139) {outputName=TS_Mask_02, sortPriority=10}
  FunctionOutput(None=@n186) {outputName=TS_Mask_01, sortPriority=10}
  n358 = FunctionInput() {inputName=Triplaners-Masks textures}
  n359 = Constant() {r=1}
  FunctionOutput(None=StaticSwitchParameter(True=@n186, False=AppendVector(A=@n358, B=@n359)) {group=Customization, parameterName=TopSideProjection}) {outputName=Mask_Override}
  n360 = ScalarParameter() {group=TopSideProjection_Emissive, parameterName="TS_Emissive----->BaseColor", sortPriority=-10}
  LinearInterpolate(A=@n344, Alpha=@n360) {constB=0}
  LinearInterpolate(A=@n329, B=@n334, Alpha=@n360)
  MakeMaterialAttributes!(
   BaseColor=@n343
   Metallic
   Specular
   Roughness
   Anisotropy
   EmissiveColor
   Opacity
   OpacityMask
   Normal
   Tangent
   WorldPositionOffset
   SubsurfaceColor
   ClearCoat
   ClearCoatRoughness
   AmbientOcclusion
   Refraction
   CustomizedUVs_0
   CustomizedUVs_1
   CustomizedUVs_2
   CustomizedUVs_3
   CustomizedUVs_4
   CustomizedUVs_5
   CustomizedUVs_6
   CustomizedUVs_7
   PixelDepthOffset
   ShadingModel
   Displacement
  )
  Multiply(
   A=@n344
   B=MaterialFunctionCall(
    Value_Morning=ScalarParameter() {defaultValue=1, desc=0.5, group=TopSideProjection__Emissive_TOD, parameterName=TS_TOD_Value_Morning}
    Value_Day=ScalarParameter() {defaultValue=1, group=TopSideProjection__Emissive_TOD, parameterName=TS_TOD_Value_Day}
    Value_Evening=ScalarParameter() {defaultValue=1, desc=0.5, group=TopSideProjection__Emissive_TOD, parameterName=TS_TOD_Value_Evening}
    Value_Night=ScalarParameter() {defaultValue=1, desc=0.25, group=TopSideProjection__Emissive_TOD, parameterName=TS_TOD_Value_Night}
   ) {materialFunction=&Scalar_Time_of_Day}.Result
  )
  n361 = OneMinus(None=@n114)
  NamedRerouteDeclaration(None=Saturate(None=Add(A=@n361, B=@n315.Result))) {name=Tri_Planner_Masking_02, nodeColor=<rgba 0 1 0.1328 1>}
  n362 = Time()
  n363 = VectorParameter() {group=TopSideProjection_UVS, parameterName=TSS_Panning_Vehicle_01, defaultValue=<rgba 0 0 0 1>}
  n364 = VectorParameter() {group=TopSideProjection_UVS, parameterName=TSS_Panning_VehicleSide_01, defaultValue=<rgba 0 0 0 1>}
  n365 = MaterialFunctionCall(Float4=NamedRerouteUsage()) {materialFunction=&BreakOutFloat4Components}
  n366 = LinearInterpolate(
   A=LinearInterpolate(
    A=VectorParameter() {group=TopSideProjection_UVS, parameterName=TSS_Panning_Weapon_01, defaultValue=<rgba 0 0 0 1>}.RGB
    B=LinearInterpolate(A=@n363.RGB, B=@n364.RGB, Alpha=@n365.G)
    Alpha=@n365.R
   )
   B=LinearInterpolate(
    A=VectorParameter() {group=TopSideProjection_UVS, parameterName=TSS_Panning_Mech_01, defaultValue=<rgba 0 0 0 1>}.RGB
    B=VectorParameter() {group=TopSideProjection_UVS, parameterName=TSS_Panning_Character_01, defaultValue=<rgba 0 0 0 1>}.RGB
    Alpha=@n365.A
   )
   Alpha=@n365.B
  )
  n367 = VectorParameter() {group=TopSideProjection_UVS, parameterName=TSS_Offset_Vehicle_01, defaultValue=<rgba 0 0 0 1>}
  n368 = VectorParameter() {group=TopSideProjection_UVS, parameterName=TSS_Offset_VehicleSide_01, defaultValue=<rgba 0 0 0 1>}
  n369 = MaterialFunctionCall(Float4=NamedRerouteUsage()) {materialFunction=&BreakOutFloat4Components}
  n370 = LinearInterpolate(
   A=LinearInterpolate(
    A=VectorParameter() {group=TopSideProjection_UVS, parameterName=TSS_Offset_Weapon_01, defaultValue=<rgba 0 0 0 1>}.RGB
    B=LinearInterpolate(A=@n367.RGB, B=@n368.RGB, Alpha=@n369.G)
    Alpha=@n369.R
   )
   B=LinearInterpolate(
    A=VectorParameter() {group=TopSideProjection_UVS, parameterName=TSS_Offset_Mech_01, defaultValue=<rgba 0 0 0 1>}.RGB
    B=VectorParameter() {group=TopSideProjection_UVS, parameterName=TSS_Offset_Character_01, defaultValue=<rgba 0 0 0 1>}.RGB
    Alpha=@n369.A
   )
   Alpha=@n369.B
  )
  n371 = VectorParameter() {group=TopSideProjection_UVS, parameterName=TSS_Offset_Pickaxe_01, defaultValue=<rgba 0 0 0 1>}
  n372 = NamedRerouteUsage()
  n373 = Add(
   A=Frac(None=Multiply(A=@n362, B=@n366))
   B=LinearInterpolate(
    A=LinearInterpolate(A=@n370, B=@n371.RGB, Alpha=@n372)
    B=VectorParameter() {group=TopSideProjection_UVS, parameterName=TSS_Offset_Backpack_01, defaultValue=<rgba 0 0 0 1>}.RGB
    Alpha=NamedRerouteUsage()
   )
  )
  NamedRerouteDeclaration(None=StaticSwitchParameter(True=Add(A=@n373, B=@n183), False=@n373) {group=TopSideProjection_UVS, parameterName=TS-WarpUVS-01---------------------, sortPriority=50}) {name=TSSP_01_Offset, nodeColor=<rgba 1 0.7969 0 1>}
  n374 = VectorParameter() {group=TopSideProjection_UVS, parameterName=TSS_Scale_Weapon_01, defaultValue=<rgba 1 1 1 1>}
  n375 = VectorParameter() {group=TopSideProjection_UVS, parameterName=TSS_Scale_Vehicle_01, defaultValue=<rgba 1 1 1 1>}
  n376 = Multiply(A=@n375.RGB, B=@n375.RGB)
  n377 = VectorParameter() {group=TopSideProjection_UVS, parameterName=TSS_Scale_VehicleSide_01, defaultValue=<rgba 1 1 1 1>}
  n378 = Multiply(A=@n377.RGB, B=@n377.RGB)
  n379 = MaterialFunctionCall(Float4=NamedRerouteUsage()) {materialFunction=&BreakOutFloat4Components}
  n380 = VectorParameter() {group=TopSideProjection_UVS, parameterName=TSS_Scale_Mech_01, defaultValue=<rgba 1 1 1 1>}
  n381 = VectorParameter() {group=TopSideProjection_UVS, parameterName=TSS_Scale_Character_01, defaultValue=<rgba 1 1 1 1>}
  n382 = LinearInterpolate(
   A=LinearInterpolate(A=Multiply(A=@n374.RGB, B=@n374.RGB), B=LinearInterpolate(A=@n376, B=@n378, Alpha=@n379.G), Alpha=@n379.R)
   B=LinearInterpolate(A=Multiply(A=@n380.RGB, B=@n380.RGB), B=Multiply(A=@n381.RGB, B=@n381.RGB), Alpha=@n379.A)
   Alpha=@n379.B
  )
  n383 = ScalarParameter() {group=TopSideProjection_UVS, parameterName=TSS_Scale_Pickaxe_01_FEMALEPICKAXE_ADJUST}
  n384 = Saturate(None=Subtract(A=ScalarParameter() {bUseCustomPrimitiveData=true, group=TopSideProjection_UVS, parameterName=TSS_Is_FemalePickAxe}) {constB=3})
  n385 = VectorParameter() {group=TopSideProjection_UVS, parameterName=TSS_Scale_Pickaxe_01, defaultValue=<rgba 1 1 1 1>}
  n386 = Multiply(A=Add(A=Multiply(A=@n383, B=@n384), B=@n385.RGB), B=@n385.A)
  n387 = NamedRerouteUsage()
  n388 = VectorParameter() {group=TopSideProjection_UVS, parameterName=TSS_Scale_Backpack_01, defaultValue=<rgba 1 1 1 1>}
  NamedRerouteDeclaration(None=LinearInterpolate(A=LinearInterpolate(A=@n382, B=@n386, Alpha=@n387), B=Multiply(A=@n388.RGB, B=@n388.RGB), Alpha=NamedRerouteUsage())) {name=TSSP_01_Scale, nodeColor=<rgba 1 0.7969 0 1>}
  NamedRerouteDeclaration(None=@n17.Result) {name=Caculated_Bounds, nodeColor=<rgba 0.125 0 1 1>}
  NamedRerouteDeclaration(None=@n127) {name=PreSkin_Normal, nodeColor=<rgba 0.1719 1 0 1>}
  NamedRerouteDeclaration(None=@n16) {name=PreSkin_Position, nodeColor=<rgba 0.1719 1 0 1>}
  n389 = ScalarParameter() {group=TopSideProjection_UVS, parameterName=TSS_Rotation_Weapon_01}
  n390 = MaterialFunctionCall(Float4=NamedRerouteUsage()) {materialFunction=&BreakOutFloat4Components}
  n391 = LinearInterpolate(
   A=ScalarParameter() {group=TopSideProjection_UVS, parameterName=TSS_Rotation_Vehicle_01}
   B=ScalarParameter() {group=TopSideProjection_UVS, parameterName=TSS_Rotation_VehicleSide_01}
   Alpha=@n390.G
  )
  n392 = ScalarParameter() {defaultValue=0.25, group=TopSideProjection_UVS, parameterName=TSS_Rotation_Mech_01}
  n393 = ScalarParameter() {defaultValue=0.25, group=TopSideProjection_UVS, parameterName=TSS_Rotation_Character_01}
  NamedRerouteDeclaration(None=LinearInterpolate(A=LinearInterpolate(A=@n389, B=@n391, Alpha=@n390.R), B=LinearInterpolate(A=@n392, B=@n393, Alpha=@n390.A), Alpha=@n390.B)) {name=TSSP_01_Rotation, nodeColor=<rgba 1 0 0.2812 1>}
  n394 = TextureObjectParameter() {
   group=TopSideProjection
   parameterName=TS_Normal_01
   samplerType=SAMPLERTYPE_Normal
   sortPriority=16
   texture=&FlatNormalBlue
  }
  NamedRerouteDeclaration(None=@n394) {name=Tex_TS01_Normal, nodeColor=<rgba 0.3125 0 1 1>}
  n395 = Add(A=@n221, B=ScalarParameter() {defaultValue=1, group=TopSideProjection_Emissive, parameterName=TS-Emissive01-CurveMask, sortPriority=4})
  n396 = ScalarParameter() {group=TopSideProjection_Emissive, parameterName=TS-Emissive-01-Color-Animated, sortPriority=3}
  n397 = Time()
  n398 = MaterialFunctionCall(Float4=NamedRerouteUsage()) {materialFunction=&BreakOutFloat4Components}
  n399 = LinearInterpolate(
   A=ScalarParameter() {group=TopSideProjection_Emissive, parameterName=TS_Emissive01Curve_Offset_Weapon, sortPriority=4}
   B=ScalarParameter() {group=TopSideProjection_Emissive, parameterName=TS_Emissive01Curve_Offset_Vehicle, sortPriority=4}
   Alpha=@n398.R
  )
  n400 = ScalarParameter() {group=TopSideProjection_Emissive, parameterName=TS_Emissive01Curve_Offset_Mech, sortPriority=4}
  n401 = ScalarParameter() {
   defaultValue=1
   desc="negative value inverts "
   group=TopSideProjection_Emissive
   parameterName=TS_Emissive01Curve_Falloff_Weapon
   sortPriority=4
  }
  n402 = ScalarParameter() {
   defaultValue=1
   desc="negative value inverts "
   group=TopSideProjection_Emissive
   parameterName=TS_Emissive01Curve_Falloff_Vehicle
   sortPriority=4
  }
  n403 = LinearInterpolate(A=@n401, B=@n402, Alpha=@n398.R)
  n404 = ScalarParameter() {
   defaultValue=1
   desc="negative value inverts "
   group=TopSideProjection_Emissive
   parameterName=TS_Emissive01Curve_Falloff_Mech
   sortPriority=4
  }
  n405 = Frac(None=MaterialFunctionCall(
    Bounds=@n17.Result
    ForwardAxis=@n76
    Offset=LinearInterpolate(A=@n399, B=@n400, Alpha=@n398.B)
    FallOff=LinearInterpolate(A=@n403, B=@n404, Alpha=@n398.B)
   ) {materialFunction=&MF_Wrap_Gradent_Simple}.Raw_Data)
  n406 = Multiply(
   A=ScalarParameter() {defaultValue=1, group=TopSideProjection_Emissive, parameterName=TS-Emissive-01-Color-TileRate, sortPriority=3}
   B=StaticSwitch(
    True=Add(A=@n189, B=@n405)
    False=Saturate(None=@n221)
    Value=StaticBoolParameter() {group=TopSideProjection_Emissive, parameterName="TS-Emissive01Curve-Height----->Uvs", sortPriority=4}
   )
  )
  n407 = ScalarParameter() {group=TopSideProjection_Emissive, parameterName=TS-Emissive-01-Color-Offset, sortPriority=3}
  n408 = Frac(None=Add(A=Multiply(A=@n396, B=@n397), B=Add(A=@n406, B=@n407)))
  n409 = CurveAtlasRowParameter(CurveTime=@n408) {
   atlas=/Game/Weapons/WeaponSkins/Wraps/Textures/CruveAtlas_Wraps_01
   curve=/Game/Weapons/WeaponSkins/Wraps/Textures/C_GradSimple
   group=TopSideProjection_Emissive
   parameterName=TS_Emissive01_Curve
   sortPriority=3
  }
  n410 = MaterialFunctionCall!(Index, CCinput=@n409, CCinput_Alpha=@n409, UV=@n408) {materialFunction=&/MF_Global_CC_Tex}
  n411 = Multiply(
   A=@n410.Alpha
   B=ScalarParameter() {defaultValue=1, group=TopSideProjection_Emissive, parameterName=TS-Emissive01-CurveBrightness, sortPriority=4}
  )
  n412 = VectorParameter() {group=TopSideProjection_Emissive, parameterName=TS_Emissive01_Color, sortPriority=3, defaultValue=<rgba 0 0 0 1>}
  n413 = StaticSwitchParameter(True=Multiply(A=Saturate(None=@n395), B=Multiply(A=@n410.3V, B=@n411)), False=Multiply(A=@n412.RGB, B=@n221)) {group=TopSideProjection_Emissive, parameterName="TS-Emissive-01-Color--->Curve", sortPriority=2}
  n414 = LinearInterpolate(
   A=ScalarParameter() {defaultValue=1, group=TopSideProjection_Emissive_Reactive, parameterName=TS_Emissive01_Time_Brightness__Start}
   B=ScalarParameter() {group=TopSideProjection_Emissive_Reactive, parameterName=TS_Emissive01_Time_Brightness__End}
   Alpha=@n90
  )
  n415 = StaticSwitchParameter(
   True=StaticSwitchParameter(True=Multiply(A=@n413, B=@n414), False=@n413) {group=TopSideProjection_Emissive_Reactive, parameterName=TS_Emissive01_Reactive_Brightness---------------------}
   False=Constant()
  ) {group=TopSideProjection_Emissive, parameterName=TS-Emissive-01------------------------------------------------------, sortPriority=1}
  n416 = StaticSwitchParameter(True=Multiply(A=@n415, B=@n326), False=@n415) {group=TopSideProjection_Emissive, parameterName=TS-Emissive-02------------------------------------------------------, sortPriority=100}
  NamedRerouteDeclaration(None=@n416) {name=Emissive_01, nodeColor=<rgba 1 0.0703 0 1>}
  n417 = Add(A=ScalarParameter() {defaultValue=1, group=TopSideProjection_Emissive, parameterName=TS-Emissive02-CurveMask, sortPriority=103}, B=@n321)
  n418 = ScalarParameter() {group=TopSideProjection_Emissive, parameterName=TS-Emissive-02-Color-Animated, sortPriority=102}
  n419 = Time()
  n420 = MaterialFunctionCall(Float4=NamedRerouteUsage()) {materialFunction=&BreakOutFloat4Components}
  n421 = LinearInterpolate(
   A=ScalarParameter() {group=TopSideProjection_Emissive, parameterName=TS_Emissive02Curve_Offset_Weapon, sortPriority=103}
   B=ScalarParameter() {group=TopSideProjection_Emissive, parameterName=TS_Emissive02Curve_Offset_Vehicle, sortPriority=103}
   Alpha=@n420.R
  )
  n422 = ScalarParameter() {group=TopSideProjection_Emissive, parameterName=TS_Emissive02Curve_Offset_Mech, sortPriority=103}
  n423 = ScalarParameter() {
   defaultValue=1
   desc="negative value inverts "
   group=TopSideProjection_Emissive
   parameterName=TS_Emissive02Curve_Falloff_Weapon
   sortPriority=103
  }
  n424 = ScalarParameter() {
   defaultValue=1
   desc="negative value inverts "
   group=TopSideProjection_Emissive
   parameterName=TS_Emissive02Curve_Falloff_Vehicle
   sortPriority=103
  }
  n425 = LinearInterpolate(A=@n423, B=@n424, Alpha=@n420.R)
  n426 = ScalarParameter() {
   defaultValue=1
   desc="negative value inverts "
   group=TopSideProjection_Emissive
   parameterName=TS_Emissive02Curve_Falloff_Mech
   sortPriority=103
  }
  n427 = Frac(None=MaterialFunctionCall(
    Bounds=@n17.Result
    ForwardAxis=@n76
    Offset=LinearInterpolate(A=@n421, B=@n422, Alpha=@n420.B)
    FallOff=LinearInterpolate(A=@n425, B=@n426, Alpha=@n420.B)
   ) {materialFunction=&MF_Wrap_Gradent_Simple}.Raw_Data)
  n428 = Multiply(
   A=ScalarParameter() {defaultValue=1, group=TopSideProjection_Emissive, parameterName=TS-Emissive-02-Color-TileRate, sortPriority=102}
   B=StaticSwitch(
    True=Add(A=@n305, B=@n427)
    False=Saturate(None=@n321)
    Value=StaticBoolParameter() {group=TopSideProjection_Emissive, parameterName="TS-Emissive02Curve-Height----->Uvs", sortPriority=103}
   )
  )
  n429 = ScalarParameter() {group=TopSideProjection_Emissive, parameterName=TS-Emissive-02-Color-Offset, sortPriority=102}
  n430 = Frac(None=Add(A=Multiply(A=@n418, B=@n419), B=Add(A=@n428, B=@n429)))
  n431 = CurveAtlasRowParameter(CurveTime=@n430) {
   atlas=&CruveAtlas_Wraps_01
   curve=&C_GradSimple
   group=TopSideProjection_Emissive
   parameterName=TS_Emissive02_Curve
   sortPriority=102
  }
  n432 = MaterialFunctionCall(
   Index=ScalarParameter() {group=TopSideProjection_Emissive, parameterName="TS_Emissive02_Curve  Index", sortPriority=102}
   CCinput=@n431
   CCinput_Alpha=@n431
   UV=@n430
  ) {materialFunction=&MF_Global_CC_Tex}
  n433 = Multiply(
   A=@n432.Alpha
   B=ScalarParameter() {defaultValue=1, group=TopSideProjection_Emissive, parameterName=TS-Emissive02-CurveBrightness, sortPriority=103}
  )
  n434 = VectorParameter() {group=TopSideProjection_Emissive, parameterName=TS_Emissive02_Color, sortPriority=102, defaultValue=<rgba 0 0 0 1>}
  n435 = StaticSwitchParameter(True=Multiply(A=Saturate(None=@n417), B=Multiply(A=@n432.3V, B=@n433)), False=Multiply(A=@n434.RGB, B=@n321)) {group=TopSideProjection_Emissive, parameterName="TS-Emissive-02-Color--->Curve", sortPriority=101}
  n436 = LinearInterpolate(
   A=ScalarParameter() {defaultValue=1, group=TopSideProjection_Emissive_Reactive, parameterName=TS_Emissive02_Time_Brightness__Start}
   B=ScalarParameter() {group=TopSideProjection_Emissive_Reactive, parameterName=TS_Emissive02_Time_Brightness__End}
   Alpha=@n324
  )
  n437 = StaticSwitchParameter(
   True=StaticSwitchParameter(True=Multiply(A=@n435, B=@n436), False=@n435) {group=TopSideProjection_Emissive_Reactive, parameterName=TS_Emissive02_Reactive_Brightness---------------------}
   False=Constant()
  ) {group=TopSideProjection_Emissive, parameterName=TS-Emissive-02------------------------------------------------------, sortPriority=100}
  NamedRerouteDeclaration(None=@n437) {name=Emissive_02, nodeColor=<rgba 1 0.7266 0 1>}
  n438 = VectorParameter() {group=TopSideProjection_UVS, parameterName=TSS_Scale_Weapon_02, sortPriority=100, defaultValue=<rgba 1 1 1 1>}
  n439 = VectorParameter() {group=TopSideProjection_UVS, parameterName=TSS_Scale_Vehicle_02, sortPriority=100, defaultValue=<rgba 1 1 1 1>}
  n440 = Multiply(A=@n439.RGB, B=@n439.RGB)
  n441 = VectorParameter() {group=TopSideProjection_UVS, parameterName=TSS_Scale_VehicleSide_02, sortPriority=100, defaultValue=<rgba 1 1 1 1>}
  n442 = Multiply(A=@n441.RGB, B=@n441.RGB)
  n443 = MaterialFunctionCall(Float4=NamedRerouteUsage()) {materialFunction=&BreakOutFloat4Components}
  n444 = VectorParameter() {group=TopSideProjection_UVS, parameterName=TSS_Scale_Mech_02, sortPriority=100, defaultValue=<rgba 1 1 1 1>}
  n445 = VectorParameter() {group=TopSideProjection_UVS, parameterName=TSS_Scale_Character_02, sortPriority=100, defaultValue=<rgba 1 1 1 1>}
  n446 = LinearInterpolate(
   A=LinearInterpolate(A=Multiply(A=@n438.RGB, B=@n438.RGB), B=LinearInterpolate(A=@n440, B=@n442, Alpha=@n443.G), Alpha=@n443.R)
   B=LinearInterpolate(A=Multiply(A=@n444.RGB, B=@n444.RGB), B=Multiply(A=@n445.RGB, B=@n445.RGB), Alpha=@n443.A)
   Alpha=@n443.B
  )
  n447 = ScalarParameter() {group=TopSideProjection_UVS, parameterName=TSS_Scale_Pickaxe_01_FEMALEPICKAXE_ADJUST}
  n448 = Saturate(None=Subtract(A=ScalarParameter() {bUseCustomPrimitiveData=true, group=TopSideProjection_UVS, parameterName=TSS_Is_FemalePickAxe}) {constB=3})
  n449 = VectorParameter() {group=TopSideProjection_UVS, parameterName=TSS_UV_Scale_Pickaxe_02, sortPriority=102, defaultValue=<rgba 1 1 1 1>}
  n450 = Multiply(A=Add(A=Multiply(A=@n447, B=@n448), B=@n449.RGB), B=@n449.A)
  n451 = Saturate(None=Multiply(A=NamedRerouteUsage(), B=ScalarParameter() {group=TopSideProjection_UVS, parameterName="TSS_UV_Pickaxe_02  1=Active", sortPriority=101}))
  n452 = VectorParameter() {group=TopSideProjection_UVS, parameterName=TSS_UV_Scale_Backpack_02, sortPriority=104, defaultValue=<rgba 1 1 1 1>}
  n453 = Multiply(A=NamedRerouteUsage(), B=ScalarParameter() {group=TopSideProjection_UVS, parameterName="TSS_UV_Backpack_02 1=Active", sortPriority=103})
  NamedRerouteDeclaration(None=LinearInterpolate(A=LinearInterpolate(A=@n446, B=@n450, Alpha=@n451), B=Multiply(A=@n452.RGB, B=@n452.RGB), Alpha=Saturate(None=@n453))) {name=TSSP_01_Scale_2, nodeColor=<rgba 1 0.7969 0 1>}
  n454 = NamedRerouteUsage()
  n455 = MaterialFunctionCall(Float4=NamedRerouteUsage()) {materialFunction=&BreakOutFloat4Components}
  n456 = NamedRerouteUsage()
  n457 = VectorParameter() {group=TopSideProjection_UVS, parameterName=TS-WarpUVS-02-Anitmated Weapon, sortPriority=207, defaultValue=<rgba 0 0 0 1>}
  n458 = VectorParameter() {group=TopSideProjection_UVS, parameterName=TS-WarpUVS-02-Anitmated Vehicle, sortPriority=208, defaultValue=<rgba 0 0 0 1>}
  n459 = Multiply(A=Time() {period=1}, B=ComponentMask(None=LinearInterpolate(A=@n457.RGB, B=@n458.RGB, Alpha=@n14)) {g=true, r=true})
  n460 = AppendVector(A=@n459, B=Constant())
  n461 = ScalarParameter() {defaultValue=1, group=TopSideProjection_UVS, parameterName=TS-WarpUVS-02-TileRate Character, sortPriority=204}
  n462 = LinearInterpolate(
   A=ScalarParameter() {defaultValue=1, group=TopSideProjection_UVS, parameterName=TS-WarpUVS-02-TileRate Weapon, sortPriority=203}
   B=Multiply(
    A=ScalarParameter() {defaultValue=1, group=TopSideProjection_UVS, parameterName=TS-WarpUVS-02-TileRate Vehicle, sortPriority=204}
    B=LinearInterpolate(B=@n461, Alpha=@n455.A) {constA=1}
   )
   Alpha=@n14
  )
  n463 = Multiply(A=TextureCoordinate(), B=@n462)
  n464 = TextureSampleParameter2D(UVs=StaticSwitchParameter(
    True=MaterialFunctionCall(
     Caculated_Bounds=@n454
     Is Vehicle=@n455.R
     Is VehicleSide=@n455.R
     Is Mech=@n455.R
     UV_Rotation Bypass=@n456
     Uvs_Offset=@n460
     Uvs_TileRate=@n462
    ) {materialFunction=&MF_Projection_Side_Normal}.UVs
    False=Add(A=@n463, B=@n459)
   ) {group=TopSideProjection_UVS, parameterName=TS-WarpUVS-Use-SideProjection-02, sortPriority=200}) {
   group=TopSideProjection_UVS
   parameterName=TS-WarpUVS-02-Texture
   sortPriority=202
   texture=&AtmosphericCloudNoise01
  }
  n465 = ScalarParameter() {group=TopSideProjection_UVS, parameterName=TS-WarpUVS-02-Amount Weapon, sortPriority=205}
  n466 = ScalarParameter() {group=TopSideProjection_UVS, parameterName=TS-WarpUVS-02-Amount_Vehicle, sortPriority=206}
  NamedRerouteDeclaration(None=Multiply(
    A=ChannelMaskParameter(None=@n464.RGBA) {group=TopSideProjection_UVS, parameterName=TS-WarpUVS-02_Channel, sortPriority=201}.R
    B=LinearInterpolate(A=@n465, B=@n466, Alpha=@n14)
   )) {name=Uvs Warp_02, nodeColor=<rgba 0 1 0.8594 1>}
  n467 = NamedRerouteUsage()
  n468 = Time()
  n469 = VectorParameter() {group=TopSideProjection_UVS, parameterName=TSS_Panning_Weapon_02, sortPriority=100, defaultValue=<rgba 0 0 0 1>}
  n470 = VectorParameter() {group=TopSideProjection_UVS, parameterName=TSS_Panning_Vehicle_02, sortPriority=100, defaultValue=<rgba 0 0 0 1>}
  n471 = VectorParameter() {group=TopSideProjection_UVS, parameterName=TSS_Panning_VehicleSide_02, sortPriority=100, defaultValue=<rgba 0 0 0 1>}
  n472 = MaterialFunctionCall(Float4=NamedRerouteUsage()) {materialFunction=&BreakOutFloat4Components}
  n473 = LinearInterpolate(A=@n469.RGB, B=LinearInterpolate(A=@n470.RGB, B=@n471.RGB, Alpha=@n472.G), Alpha=@n472.R)
  n474 = VectorParameter() {group=TopSideProjection_UVS, parameterName=TSS_Panning_Mech_02, sortPriority=100, defaultValue=<rgba 0 0 0 1>}
  n475 = VectorParameter() {group=TopSideProjection_UVS, parameterName=TSS_Panning_Character_02, sortPriority=100, defaultValue=<rgba 0 0 0 1>}
  n476 = LinearInterpolate(A=@n474.RGB, B=@n475.RGB, Alpha=@n472.A)
  n477 = VectorParameter() {group=TopSideProjection_UVS, parameterName=TSS_Panning_Pickaxe_02, sortPriority=102, defaultValue=<rgba 0 0 0 1>}
  n478 = Saturate(None=Multiply(A=NamedRerouteUsage(), B=ScalarParameter() {group=TopSideProjection_UVS, parameterName="TSS_UV_Pickaxe_02  1=Active", sortPriority=101}))
  n479 = VectorParameter() {group=TopSideProjection_UVS, parameterName=TSS_Panning_Backpack_02, sortPriority=104, defaultValue=<rgba 0 0 0 1>}
  n480 = Saturate(None=Multiply(A=NamedRerouteUsage(), B=ScalarParameter() {group=TopSideProjection_UVS, parameterName="TSS_UV_Backpack_02 1=Active", sortPriority=103}))
  n481 = LinearInterpolate(A=LinearInterpolate(A=LinearInterpolate(A=@n473, B=@n476, Alpha=@n472.B), B=@n477.RGB, Alpha=@n478), B=@n479.RGB, Alpha=@n480)
  n482 = VectorParameter() {group=TopSideProjection_UVS, parameterName=TSS_Offset_Weapon_02, sortPriority=100, defaultValue=<rgba 0 0 0 1>}
  n483 = VectorParameter() {group=TopSideProjection_UVS, parameterName=TSS_Offset_Vehicle_02, sortPriority=100, defaultValue=<rgba 0 0 0 1>}
  n484 = VectorParameter() {group=TopSideProjection_UVS, parameterName=TSS_Offset_VehicleSide_02, sortPriority=100, defaultValue=<rgba 0 0 0 1>}
  n485 = MaterialFunctionCall(Float4=NamedRerouteUsage()) {materialFunction=&BreakOutFloat4Components}
  n486 = VectorParameter() {group=TopSideProjection_UVS, parameterName=TSS_Offset_Mech_02, sortPriority=100, defaultValue=<rgba 0 0 0 1>}
  n487 = VectorParameter() {group=TopSideProjection_UVS, parameterName=TSS_Offset_Character_02, sortPriority=100, defaultValue=<rgba 0 0 0 1>}
  n488 = LinearInterpolate(
   A=LinearInterpolate(A=@n482.RGB, B=LinearInterpolate(A=@n483.RGB, B=@n484.RGB, Alpha=@n485.G), Alpha=@n485.R)
   B=LinearInterpolate(A=@n486.RGB, B=@n487.RGB, Alpha=@n485.A)
   Alpha=@n485.B
  )
  n489 = VectorParameter() {group=TopSideProjection_UVS, parameterName=TSS_UV_Offset_Pickaxe_02, sortPriority=102, defaultValue=<rgba 0 0 0 1>}
  n490 = VectorParameter() {group=TopSideProjection_UVS, parameterName=TSS_UV_Offset_Backpack_02, sortPriority=104, defaultValue=<rgba 0 0 0 1>}
  n491 = Add(
   A=Frac(None=Multiply(A=@n468, B=@n481))
   B=LinearInterpolate(A=LinearInterpolate(A=@n488, B=@n489.RGB, Alpha=@n478), B=@n490.RGB, Alpha=@n480)
  )
  NamedRerouteDeclaration(None=StaticSwitchParameter(True=Add(A=@n491, B=@n467), False=@n491) {group=TopSideProjection_UVS, parameterName=TS-WarpUVS-02---------------------, sortPriority=200}) {name=Uvs Side Offset 02, nodeColor=<rgba 1 0 0.4453 1>}
  n492 = ScalarParameter() {group=TopSideProjection_UVS, parameterName=TSS_Rotation_Weapon_02, sortPriority=100}
  n493 = MaterialFunctionCall(Float4=NamedRerouteUsage()) {materialFunction=&BreakOutFloat4Components}
  n494 = LinearInterpolate(
   A=ScalarParameter() {group=TopSideProjection_UVS, parameterName=TSS_Rotation_Vehicle_02, sortPriority=100}
   B=ScalarParameter() {group=TopSideProjection_UVS, parameterName=TSS_Rotation_VehicleSide_02, sortPriority=100}
   Alpha=@n493.G
  )
  n495 = ScalarParameter() {defaultValue=0.25, group=TopSideProjection_UVS, parameterName=TSS_Rotation_Mech_02, sortPriority=100}
  n496 = ScalarParameter() {defaultValue=0.25, group=TopSideProjection_UVS, parameterName=TSS_Rotation_Character_02, sortPriority=100}
  NamedRerouteDeclaration(None=LinearInterpolate(A=LinearInterpolate(A=@n492, B=@n494, Alpha=@n493.R), B=LinearInterpolate(A=@n495, B=@n496, Alpha=@n493.A), Alpha=@n493.B)) {name=Uvs Side Rotation 02, nodeColor=<rgba 0 0.9766 1 1>}
  n497 = NamedRerouteDeclaration(None=FunctionInput() {bUsePreviewValueAsDefault=true, inputName=IsStatsB, inputType=FunctionInput_Vector2, sortPriority=78}) {name=Is_StatsB, nodeColor=<rgba 0 0.9253 1 1>}
  NamedRerouteDeclaration(None=ComponentMask(None=@n497) {g=true}) {name=IsPickaxe, nodeColor=<rgba 1 0 0.5625 1>}
  NamedRerouteDeclaration(None=ComponentMask(None=@n497) {r=true}) {name=IsBackpack, nodeColor=<rgba 0.2656 0 1 1>}
  n498 = OneMinus(None=@n162)
  NamedRerouteDeclaration(None=Saturate(None=Add(A=@n498, B=@n204.Result))) {name=Tri_Planner_Masking_01, nodeColor=<rgba 1 0.9141 0 1>}
  PreSkinnedLocalBounds()
  n499 = NamedRerouteDeclaration(None=FunctionInput() {bUsePreviewValueAsDefault=true, inputName=isCharacter, inputType=FunctionInput_Scalar, sortPriority=100}) {name=IsCharacter, nodeColor=<rgba 1 0 0.8672 1>}
  Reroute(None=NamedRerouteDeclaration(None=MaterialFunctionCall(X=@n14, Y=@n191, Z=@n15, A=@n499) {materialFunction=&MakeFloat4}.Result) {name=IsState, nodeColor=<rgba 0.9219 1 0 1>})
  Reroute(None=@n17.Result)
  Reroute(None=Add(A=@n416, B=@n437))
  Reroute(None=StaticSwitchParameter(True=@n77.Clamped_Data, False=@n269.Clamped_Data) {group=FadingMosaic, parameterName="FD_testing Contrast --->Uv", sortPriority=-50})
  ScalarParameter() {defaultValue=1, group=TopSideProjection_Emissive, parameterName=TS-Emissive01-CurveMask, sortPriority=4}
  ScalarParameter() {defaultValue=1, group=TopSideProjection_Emissive, parameterName="TS_Emissive01_Curve  Index", sortPriority=3}
  TextureSampleParameter2D() {
   group=TopSideProjection
   parameterName=Ts_BlendGradent_Texture
   samplerType=SAMPLERTYPE_Masks
   sortPriority=75
   texture=&T_VertGradient_MASK
  }
  TextureSampleParameter2D(UVs=@n307.Uvs) {
   group=TopSideProjection_Emissive
   parameterName=TS_Emissive02_Gradent_Texture
   samplerType=SAMPLERTYPE_Masks
   sortPriority=111
   texture=&T_VertGradient_MASK
  }
 ]}
