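# Material function for weapon wraps that layers a time-of-day top-side color effect onto incoming material attributes, behind a static switch.
# Each RGBA channel of one of two switchable masks drives a morning/day/evening/night color resolved by Wrap_Color_Time_of_Day and added to emissive.
# Color alphas control how much masked areas darken existing base color and emissive, and the customization mask blends the result.
# The Color_Time_of_Day calls and Darken_Base branch do not reach the output.
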MaterialFunction /Game/Weapons/WeaponSkins/Wraps/MaterialFunctions/MF_TopSide_Effect_Mask_TOD
graph {nodes [
  n1 = FunctionInput() {inputName=Material, inputType=FunctionInput_MaterialAttributes}
  n2 = GetMaterialAttributes(None=@n1) {attributeGetTypes [#1, #2]}
  n3 = MaterialFunctionCall(Float4=StaticSwitchParameter(
    True=FunctionInput() {inputName=TopSide_Mask_02, inputType=FunctionInput_Vector4, sortPriority=2}
    False=FunctionInput() {inputName=TopSide_Mask_01, inputType=FunctionInput_Vector4, sortPriority=1}
   ) {group=TopSide_Effect_Mask_TOD, parameterName="TS_TOD TS_Mask_01 ----> TS_Mask_02", sortPriority=-8}) {materialFunction=&BreakOutFloat4Components}
  n4 = VectorParameter() {
   desc=Alpha Darkens Base
   group=TopSide_Effect_Mask_TOD
   parameterName=TS_TOD_Red_Channel_Color_Morning
   sortPriority=1
   defaultValue=<rgba 1 0 0 1>
  }
  n5 = VectorParameter() {
   desc=Alpha Darkens Base
   group=TopSide_Effect_Mask_TOD
   parameterName=TS_TOD_Red_Channel_Color_Day
   sortPriority=2
   defaultValue=<rgba 1 0 0 1>
  }
  n6 = VectorParameter() {
   desc=Alpha Darkens Base
   group=TopSide_Effect_Mask_TOD
   parameterName=TS_TOD_Red_Channel_Color_Evening
   sortPriority=3
   defaultValue=<rgba 0 0 0 1>
  }
  n7 = VectorParameter() {
   desc=Alpha Darkens Base
   group=TopSide_Effect_Mask_TOD
   parameterName=TS_TOD_Red_Channel_Color_Night
   sortPriority=4
   defaultValue=<rgba 0 0 0 1>
  }
  n8 = Saturate(None=Add(
    A=OneMinus(None=@n3.R)
    B=MaterialFunctionCall!(Color_Morning=@n4.A, Color_Day=@n5.A, Color_Evening=@n6.A, Color_Night=@n7.A) {materialFunction=&/Wrap_Color_Time_of_Day}.Result
   ))
  n9 = VectorParameter() {
   desc=Alpha Darkens Base
   group=TopSide_Effect_Mask_TOD
   parameterName=TS_TOD_Blue_Channel_Color_Morning
   sortPriority=11
   defaultValue=<rgba 0.0641 0 1 1>
  }
  n10 = VectorParameter() {
   desc=Alpha Darkens Base
   group=TopSide_Effect_Mask_TOD
   parameterName=TS_TOD_Blue_Channel_Color_Day
   sortPriority=12
   defaultValue=<rgba 0.0641 0 1 1>
  }
  n11 = VectorParameter() {
   desc=Alpha Darkens Base
   group=TopSide_Effect_Mask_TOD
   parameterName=TS_TOD_Blue_Channel_Color_Evening
   sortPriority=13
   defaultValue=<rgba 0 0 0 1>
  }
  n12 = VectorParameter() {
   desc=Alpha Darkens Base
   group=TopSide_Effect_Mask_TOD
   parameterName=TS_TOD_Blue_Channel_Color_Night
   sortPriority=14
   defaultValue=<rgba 0 0 0 1>
  }
  n13 = Saturate(None=Add(
    A=OneMinus(None=@n3.B)
    B=MaterialFunctionCall(Color_Morning=@n9.A, Color_Day=@n10.A, Color_Evening=@n11.A, Color_Night=@n12.A) {materialFunction=&Wrap_Color_Time_of_Day}.Result
   ))
  n14 = VectorParameter() {
   desc=Alpha Darkens Base
   group=TopSide_Effect_Mask_TOD
   parameterName=TS_TOD_Green_Channel_Color_Morning
   sortPriority=21
   defaultValue=<rgba 0 1 0 1>
  }
  n15 = VectorParameter() {
   desc=Alpha Darkens Base
   group=TopSide_Effect_Mask_TOD
   parameterName=TS_TOD_Green_Channel_Color_Day
   sortPriority=22
   defaultValue=<rgba 0 1 0 1>
  }
  n16 = VectorParameter() {
   desc=Alpha Darkens Base
   group=TopSide_Effect_Mask_TOD
   parameterName=TS_TOD_Green_Channel_Color_Evening
   sortPriority=23
   defaultValue=<rgba 0 0 0 1>
  }
  n17 = VectorParameter() {
   desc=Alpha Darkens Base
   group=TopSide_Effect_Mask_TOD
   parameterName=TS_TOD_Green_Channel_Color_Night
   sortPriority=24
   defaultValue=<rgba 0 0 0 1>
  }
  n18 = Saturate(None=Add(
    A=OneMinus(None=@n3.G)
    B=MaterialFunctionCall(Color_Morning=@n14.A, Color_Day=@n15.A, Color_Evening=@n16.A, Color_Night=@n17.A) {materialFunction=&Wrap_Color_Time_of_Day}.Result
   ))
  n19 = VectorParameter() {
   desc=Alpha Darkens Base
   group=TopSide_Effect_Mask_TOD
   parameterName=TS_TOD_Alpha_Channel_Color_Morning
   sortPriority=31
   defaultValue=<rgba 0 0 0 1>
  }
  n20 = VectorParameter() {desc=Alpha Darkens Base, group=TopSide_Effect_Mask_TOD, parameterName=TS_TOD_Alpha_Channel_Color_Day, defaultValue=<rgba 0 0 0 1>}
  n21 = VectorParameter() {
   desc=Alpha Darkens Base
   group=TopSide_Effect_Mask_TOD
   parameterName=TS_TOD_Alpha_Channel_Color_Evening
   sortPriority=33
   defaultValue=<rgba 0 0 0 1>
  }
  n22 = VectorParameter() {
   desc=Alpha Darkens Base
   group=TopSide_Effect_Mask_TOD
   parameterName=TS_TOD_Alpha_Channel_Color_Night
   sortPriority=34
   defaultValue=<rgba 0 0 0 1>
  }
  n23 = Saturate(None=Add(
    A=OneMinus(None=@n3.A)
    B=MaterialFunctionCall(Color_Morning=@n19.A, Color_Day=@n20.A, Color_Evening=@n21.A, Color_Night=@n22.A) {materialFunction=&Wrap_Color_Time_of_Day}.Result
   ))
  n24 = Saturate(None=Multiply(A=Multiply(A=@n8, B=@n13), B=Multiply(A=@n18, B=@n23)))
  n25 = Add(
   A=Multiply(
    A=@n3.R
    B=MaterialFunctionCall(Color_Morning=@n4.RGB, Color_Day=@n5.RGB, Color_Evening=@n6.RGB, Color_Night=@n7.RGB) {materialFunction=&Wrap_Color_Time_of_Day}.Result
   )
   B=Multiply(
    A=@n3.G
    B=MaterialFunctionCall(Color_Morning=@n14.RGB, Color_Day=@n15.RGB, Color_Evening=@n16.RGB, Color_Night=@n17.RGB) {materialFunction=&Wrap_Color_Time_of_Day}.Result
   )
  )
  n26 = Add(
   A=Multiply(
    A=@n3.B
    B=MaterialFunctionCall(Color_Morning=@n9.RGB, Color_Day=@n10.RGB, Color_Evening=@n11.RGB, Color_Night=@n12.RGB) {materialFunction=&Wrap_Color_Time_of_Day}.Result
   )
   B=Multiply(
    A=@n3.A
    B=MaterialFunctionCall(Color_Morning=@n19.RGB, Color_Day=@n20.RGB, Color_Evening=@n21.RGB, Color_Night=@n22.RGB) {materialFunction=&Wrap_Color_Time_of_Day}.Result
   )
  )
  n27 = SetMaterialAttributes!(
   MaterialAttributes=@n2.MaterialAttributes
   BaseColor=Multiply(A=@n2.BaseColor, B=@n24)
   EmissiveColor=Add(A=Multiply(A=@n2.EmissiveColor, B=@n24), B=Add(A=@n25, B=@n26))
  ) {attributeSetTypes [#1, #2]}
  n28 = Saturate(None=Add(
    A=FunctionInput() {bUsePreviewValueAsDefault=true, inputName="Customization Mask (_CM B)", inputType=FunctionInput_Scalar, sortPriority=24}
    B=ScalarParameter() {group=TopSide_Effect_Mask_TOD, parameterName=TS_TOD_Remove_Customization Mask, sortPriority=-10}
   ))
  FunctionOutput(None=StaticSwitchParameter(True=BlendMaterialAttributes(A=@n1, B=@n27, Alpha=@n28), False=@n1) {group=Customization, parameterName=TopSide_Effect_Mask_TOD})
  MaterialFunctionCall(Color_Morning=@n4.A, Color_Day=@n5.A, Color_Evening=@n6.A, Color_Night=@n7.A) {materialFunction=&Color_Time_of_Day}
  MaterialFunctionCall(Color_Morning=@n4.RGB, Color_Day=@n5.RGB, Color_Evening=@n6.RGB, Color_Night=@n7.RGB) {materialFunction=&Color_Time_of_Day}
  MaterialFunctionCall(Color_Morning=@n9.RGB, Color_Day=@n10.RGB, Color_Evening=@n11.RGB, Color_Night=@n12.RGB) {materialFunction=&Color_Time_of_Day}
  MaterialFunctionCall(Color_Morning=@n14.RGB, Color_Day=@n15.RGB, Color_Evening=@n16.RGB, Color_Night=@n17.RGB) {materialFunction=&Color_Time_of_Day}
  MaterialFunctionCall(Color_Morning=@n19.RGB, Color_Day=@n20.RGB, Color_Evening=@n21.RGB, Color_Night=@n22.RGB) {materialFunction=&Color_Time_of_Day}
  MaterialFunctionCall(Color_Morning=@n14.A, Color_Day=@n15.A, Color_Evening=@n16.A, Color_Night=@n17.A) {materialFunction=&Color_Time_of_Day}
  MaterialFunctionCall(Color_Morning=@n9.A, Color_Day=@n10.A, Color_Evening=@n11.A, Color_Night=@n12.A) {materialFunction=&Color_Time_of_Day}
  MaterialFunctionCall(Color_Morning=@n19.A, Color_Day=@n20.A, Color_Evening=@n21.A, Color_Night=@n22.A) {materialFunction=&Color_Time_of_Day}
  n29 = ScalarParameter() {group=TopSide_Effect_Mask_TOD, parameterName=TS_TOD_Darken_Base, sortPriority=-9}
  n30 = Saturate(None=Add(A=Add(A=@n3.R, B=@n3.R), B=@n3.B))
  Reroute(None=OneMinus(None=Multiply(A=@n29, B=@n30)))
 ]}
