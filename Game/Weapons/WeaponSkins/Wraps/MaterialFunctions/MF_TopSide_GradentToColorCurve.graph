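# Material function for weapon, vehicle and mech wraps that remaps a projected mask texture's gradient to animated color-curve colors.
# It projects the mask top/side or side-only from pre-skinned normals and bounds, using per-type transforms.
# One channel plus time indexes the curve atlas; another masks base color and optional specular, with an optional Z-facing override.
# A customization mask blends the result; a static switch bypasses the effect.
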MaterialFunction /Game/Weapons/WeaponSkins/Wraps/MaterialFunctions/MF_TopSide_GradentToColorCurve
graph {nodes [
  n1 = FunctionInput(Preview=VertexInterpolator(VS=PreSkinnedNormal()).PS) {bUsePreviewValueAsDefault=true, inputName=Pre-Skinned Local Normal, sortPriority=10}
  n2 = Multiply(A=ComponentMask(None=@n1) {b=true}) {constB=-1}
  Clamp(None=@n2)
  n3 = FunctionInput() {inputName=Material, inputType=FunctionInput_MaterialAttributes, sortPriority=-10}
  n4 = GetMaterialAttributes(None=@n3) {attributeGetTypes [#1, #2, #3]}
  n5 = VectorParameter() {group=TopSide_GradentToColorCurve, parameterName=TSCC_Fill_alternate_Axis_Color_Mask_01, sortPriority=5, defaultValue=<rgba 0 0 0 1>}
  n6 = AppendVector(A=@n5.RGB, B=@n5.RGB)
  n7 = FunctionInput() {inputName=isVehicle, inputType=FunctionInput_Scalar, sortPriority=5}
  n8 = FunctionInput() {bUsePreviewValueAsDefault=true, inputName=Side Project Vehicle, inputType=FunctionInput_Scalar}
  n9 = FunctionInput() {bUsePreviewValueAsDefault=true, inputName=isMech, inputType=FunctionInput_Scalar, sortPriority=4}
  n10 = MaterialFunctionCall(R=@n7, G=@n8, B=@n9) {materialFunction=&AppendMany}
  n11 = MaterialFunctionCall(Color=@n10.RGB) {materialFunction=&SplitComponents}
  n12 = ScalarParameter() {group=TopSide_GradentToColorCurve_UVS, parameterName=TSCC_Rotation_Side_Vehicle_01}
  n13 = ScalarParameter() {group=TopSide_GradentToColorCurve_UVS, parameterName=TSCC_Rotation_Side_VehicleSide_01}
  n14 = MaterialFunctionCall(Color=@n10.RGB) {materialFunction=&SplitComponents}
  n15 = LinearInterpolate(
   A=LinearInterpolate(
    A=ScalarParameter() {group=TopSide_GradentToColorCurve_UVS, parameterName=TSCC_Rotation_Side_Weapon_01}
    B=LinearInterpolate(A=@n12, B=@n13, Alpha=@n14.G)
    Alpha=@n14.R
   )
   B=ScalarParameter() {group=TopSide_GradentToColorCurve_UVS, parameterName=TSCC_Rotation_Side_Mech_01}
   Alpha=@n14.B
  )
  n16 = VectorParameter() {group=TopSide_GradentToColorCurve_UVS, parameterName=TSCC_Offset_Side_Weapon_01, defaultValue=<rgba 0 0 0 1>}
  n17 = VectorParameter() {group=TopSide_GradentToColorCurve_UVS, parameterName=TSCC_Offset_Side_Vehicle_01, defaultValue=<rgba 0 0 0 1>}
  n18 = VectorParameter() {group=TopSide_GradentToColorCurve_UVS, parameterName=TSCC_Offset_Side_Vehicle_Side_01, defaultValue=<rgba 0 0 0 1>}
  n19 = MaterialFunctionCall(Color=@n10.RGB) {materialFunction=&SplitComponents}
  n20 = VectorParameter() {group=TopSide_GradentToColorCurve_UVS, parameterName=TSCC_Offset_Side_Mech_01, defaultValue=<rgba 0 0 0 1>}
  n21 = LinearInterpolate(A=LinearInterpolate(A=@n16.RGB, B=LinearInterpolate(A=@n17.RGB, B=@n18.RGB, Alpha=@n19.G), Alpha=@n19.R), B=@n20.RGB, Alpha=@n19.B)
  n22 = VectorParameter() {group=TopSide_GradentToColorCurve_UVS, parameterName="TSCC_Scale_Side_Weapon_01 A=uniform", defaultValue=<rgba 1 1 1 1>}
  n23 = VectorParameter() {group=TopSide_GradentToColorCurve_UVS, parameterName="TSCC_Scale_Side_Vehicle_01 A=uniform", defaultValue=<rgba 1 1 1 1>}
  n24 = Multiply(A=@n23.RGB, B=@n23.RGB)
  n25 = VectorParameter() {
   desc=must activate TS_Vehicle_Side_01 UVs
   group=TopSide_GradentToColorCurve_UVS
   parameterName="TSCC_Scale_Side_Vehicle_Side_01 A=uniform"
   defaultValue=<rgba 1 1 1 1>
  }
  n26 = Multiply(A=@n25.RGB, B=@n25.RGB)
  n27 = MaterialFunctionCall(Color=@n10.RGB) {materialFunction=&SplitComponents}
  n28 = VectorParameter() {group=TopSide_GradentToColorCurve_UVS, parameterName="TSCC_Scale_Side_Mech_01 A=uniform", defaultValue=<rgba 1 1 1 1>}
  n29 = LinearInterpolate(
   A=LinearInterpolate(A=Multiply(A=@n22.RGB, B=@n22.RGB), B=LinearInterpolate(A=@n24, B=@n26, Alpha=@n27.G), Alpha=@n27.R)
   B=Multiply(A=@n28.RGB, B=@n28.RGB)
   Alpha=@n27.B
  )
  n30 = VertexInterpolator(VS=PreSkinnedPosition())
  n31 = PreSkinnedLocalBounds()
  n32 = FunctionInput(Preview=Divide(A=Subtract(A=@n30.PS, B=@n31.Min), B=Subtract(A=@n31.Max, B=@n31.Max))) {bUsePreviewValueAsDefault=true, inputName=Calcated_Bounds, sortPriority=11}
  n33 = TextureObjectParameter() {
   group=TopSide_GradentToColorCurve
   parameterName=TSCC_Mask_01
   samplerType=SAMPLERTYPE_Masks
   sortPriority=5
   texture=&T_BlackMask
  }
  n34 = ScalarParameter() {group=TopSide_GradentToColorCurve, parameterName=TSCC_Fill_alternate_Axis_Opacity 01, sortPriority=5}
  n35 = OneMinus(None=@n34)
  n36 = MaterialFunctionCall(Color=Abs(None=@n1)) {materialFunction=&SplitComponents}
  n37 = MaterialFunctionCall(Color=@n10.RGB) {materialFunction=&SplitComponents}
  n38 = LinearInterpolate(A=@n36.B, B=@n36.B, Alpha=@n37.G)
  n39 = ScalarParameter() {defaultValue=1, group=TopSide_GradentToColorCurve, parameterName=TSCC_Blend_Contrast, sortPriority=6}
  n40 = MaterialFunctionCall(In=LinearInterpolate(A=LinearInterpolate(A=@n36.R, B=@n38, Alpha=@n37.R), B=@n36.G, Alpha=@n37.B), Contrast=@n39) {materialFunction=&CheapContrast}
  n41 = LinearInterpolate(
   A=@n6
   B=TextureSample(
    UVs=MaterialFunctionCall!(
     For_Safe_Zones
     Normal_Map
     Caculated_Bounds=@n32
     Pre_Skinned_Normals=@n1
     Is Vehicle=@n11.R
     Is VehicleSide=@n11.R
     Is Mech=@n11.R
     Vehicle_Rotation
     UV_Rotation Bypass=@n15
     Mech_Rotation
     Uvs_Offset=@n21
     Offset-after rotation
     Uvs_TileRate=@n29
    ) {materialFunction=&/MF_Projection_Side_Normal}.UVs
    Tex=@n33
   ).RGBA
   Alpha=Saturate(None=Add(A=@n35, B=@n40.Result))
  )
  n42 = VectorParameter() {
   group=TopSide_GradentToColorCurve_UVS
   parameterName=TSCC_Rotation-Weapon-01---------------------
   sortPriority=1
   defaultValue=<rgba 0 0 -0.25 1>
  }
  n43 = VectorParameter() {group=TopSide_GradentToColorCurve_UVS, parameterName=TSCC_Rotation_Vehicle_01, sortPriority=2, defaultValue=<rgba 0 0 0 1>}
  n44 = VectorParameter() {group=TopSide_GradentToColorCurve_UVS, parameterName=TSCC_Rotation_Vehicle_Side_01, sortPriority=2, defaultValue=<rgba 0 0 0 1>}
  n45 = LinearInterpolate(A=@n43.RGB, B=@n44.RGB, Alpha=@n8)
  n46 = VectorParameter() {group=TopSide_GradentToColorCurve_UVS, parameterName=TSCC_Rotation_Mech_01, sortPriority=3, defaultValue=<rgba 0.25 0.25 0 1>}
  n47 = VectorParameter() {
   group=TopSide_GradentToColorCurve_UVS
   parameterName=TSCC_Offset_Weapon_01---------------------
   sortPriority=10
   defaultValue=<rgba 0 0 0 1>
  }
  n48 = VectorParameter() {group=TopSide_GradentToColorCurve_UVS, parameterName=TSCC_Offset_Vehicle_01, sortPriority=11, defaultValue=<rgba 0 0 0 1>}
  n49 = VectorParameter() {group=TopSide_GradentToColorCurve_UVS, parameterName=TSCC_Offset_Vehicle_Side_01, sortPriority=11, defaultValue=<rgba 0 0 0 1>}
  n50 = VectorParameter() {group=TopSide_GradentToColorCurve_UVS, parameterName=TF_Offset_Mech_01, sortPriority=12, defaultValue=<rgba 0 0 0 1>}
  n51 = LinearInterpolate(A=LinearInterpolate(A=@n47.RGB, B=LinearInterpolate(A=@n48.RGB, B=@n49.RGB, Alpha=@n8), Alpha=@n7), B=@n50.RGB, Alpha=@n9)
  n52 = VectorParameter() {group=TopSide_GradentToColorCurve_UVS, parameterName="TSCC_Scale_Weapon_01 A=uniform", sortPriority=30, defaultValue=<rgba 1 1 1 1>}
  n53 = VectorParameter() {group=TopSide_GradentToColorCurve_UVS, parameterName="TSCC_Scale_Vehicle_01 A=uniform", sortPriority=31, defaultValue=<rgba 1 1 1 1>}
  n54 = VectorParameter() {
   desc=must activate TS_Vehicle_Side_01 UVs
   group=TopSide_GradentToColorCurve_UVS
   parameterName="TSCC_Scale_Vehicle_Side_01 A=uniform"
   sortPriority=31
   defaultValue=<rgba 1 1 1 1>
  }
  n55 = LinearInterpolate(
   A=Multiply(A=@n52.RGB, B=@n52.RGB)
   B=LinearInterpolate(A=Multiply(A=@n53.RGB, B=@n53.RGB), B=Multiply(A=@n54.RGB, B=@n54.RGB), Alpha=@n8)
   Alpha=@n7
  )
  n56 = VectorParameter() {group=TopSide_GradentToColorCurve_UVS, parameterName="TSCC_Scale_Mech_01 A=uniform", defaultValue=<rgba 1 1 1 1>}
  n57 = Multiply(A=@n56.RGB, B=@n56.RGB)
  n58 = MaterialFunctionCall!(
   IsCharacter
   IsStatsB
   Texture=@n33
   Rotation_Axis=LinearInterpolate(A=LinearInterpolate(A=@n42.RGB, B=@n45, Alpha=@n7), B=@n46.RGB, Alpha=@n9)
   Caculated_Bounds=Multiply(A=Add(A=@n32, B=@n51), B=LinearInterpolate(A=@n55, B=@n57, Alpha=@n9))
   Pre_Skinned_Normals=@n1
   Blend_Contrast=LinearInterpolate(
    A=@n39
    B=ScalarParameter() {defaultValue=2, group=TopSide_GradentToColorCurve, parameterName=TSCC_Blend_Contrast_Mech, sortPriority=6}
    Alpha=@n9
   )
   Fill_alternate_Axis_Opacity=@n34
   "Fill_alternate_Axis_Color_A=Roughness"=@n6
   Is_Vehicle=@n7
   Is_Mech=@n9
   Enable_Fill_Z_Weapon
   Fill_Z_Weapon_opacity
   Fill_Z_Weapon_Color
  ) {materialFunction=&/MF_TopsSideProjection_Base}
  n59 = StaticSwitchParameter(True=@n41, False=@n58.Combined) {group=TopSide_GradentToColorCurve, parameterName=TSCC_Use_Side_Projection, sortPriority=0}
  n60 = MaterialFunctionCall(In=@n2, Contrast=ScalarParameter() {defaultValue=1, group=TopSide_GradentToColorCurve, parameterName=TSCC_Mask_Z_Contrast, sortPriority=11}) {materialFunction=&CheapContrast}
  n61 = Saturate(None=Add(
    A=OneMinus(None=@n60.Result)
    B=ScalarParameter() {defaultValue=1, group=TopSide_GradentToColorCurve, parameterName="TSCC_Mask_Z 0=active 1=off", sortPriority=11}
   ))
  n62 = Multiply(
   A=LinearInterpolate(
    A=ScalarParameter() {group=TopSide_GradentToColorCurve, parameterName=TSCC_Mask_01_Channel_ColorCuve_mask_Z, sortPriority=20}
    B=ChannelMaskParameter(None=@n59) {group=TopSide_GradentToColorCurve, parameterName=TSCC_Mask_01_Channel_ColorCuve, sortPriority=20}.R
    Alpha=@n61
   )
   B=ScalarParameter() {defaultValue=1, group=TopSide_GradentToColorCurve, parameterName=TSCC_ColorCurve_Tile, sortPriority=22}
  )
  n63 = LinearInterpolate(
   A=ScalarParameter() {group=TopSide_GradentToColorCurve, parameterName=TSCC_ColorCurve_Offset, sortPriority=22}
   Alpha=ScalarParameter() {group=TopSide_GradentToColorCurve, parameterName="TSCC_ColorCurve_Disable_Reactive 0=active 1=off", sortPriority=22}
  ) {constB=0}
  n64 = Time()
  n65 = ScalarParameter() {group=TopSide_GradentToColorCurve, parameterName=TSCC_ColorCurve_Speed, sortPriority=22}
  n66 = Frac(None=Add(A=Add(A=@n62, B=@n63), B=Multiply(A=@n64, B=@n65)))
  n67 = CurveAtlasRowParameter(CurveTime=@n66) {
   atlas=/Game/Weapons/WeaponSkins/Wraps/Textures/CruveAtlas_Wraps_01
   curve=/Game/Weapons/WeaponSkins/Wraps/Textures/NeonBands/T_RainbowLavaB
   defaultValue=17
   group=TopSide_GradentToColorCurve
   parameterName=TSCC_ColorCurve_
   sortPriority=21
  }
  n68 = Multiply(
   A=MaterialFunctionCall!(
    Index=ScalarParameter() {defaultValue=17, group=TopSide_GradentToColorCurve, parameterName="TSCC_ColorCurve_  Index", sortPriority=21}
    CCinput=@n67
    CCinput_Alpha
    UV=@n66
   ) {materialFunction=&/MF_Global_CC_Tex}.3V
   B=ScalarParameter() {defaultValue=1, group=TopSide_GradentToColorCurve, parameterName=TSCC_ColorCurve_Brightness, sortPriority=22}
  )
  n69 = ScalarParameter() {group=TopSide_GradentToColorCurve, parameterName="TSCC_Mask_01_Use Channel to mask CC  0=Active 1=off", sortPriority=10}
  n70 = ChannelMaskParameter(None=StaticSwitchParameter(True=@n41, False=@n58.Combined) {group=TopSide_GradentToColorCurve, parameterName=TSCC_Use_Side_Projection, sortPriority=0}) {
   group=TopSide_GradentToColorCurve
   maskChannel=Alpha
   parameterName=TSCC_Mask_01_Channel_Mask_ColorCuve
   sortPriority=11
   defaultValue=<rgba 0 0 0 1>
  }
  n71 = LinearInterpolate(
   A=ScalarParameter() {defaultValue=1, group=TopSide_GradentToColorCurve, parameterName=TSCC_Mask_01_Channel_Mask_Z, sortPriority=12}
   B=Saturate(None=Add(A=@n69, B=@n70.R))
   Alpha=@n61
  )
  n72 = ScalarParameter() {defaultValue=0.5, group=TopSide_GradentToColorCurve, parameterName=TSCC_Specular_Masked_Section, sortPriority=1}
  n73 = Saturate(None=Multiply(
    A=ScalarParameter() {group=TopSide_GradentToColorCurve, parameterName="TSCC_Specular_Masked_Section_use 1=active 0=off", sortPriority=1}
    B=@n71
   ))
  n74 = FunctionInput() {bUsePreviewValueAsDefault=true, inputName="Customization Mask (_CM B)", inputType=FunctionInput_Scalar, sortPriority=24}
  n75 = ScalarParameter() {group=TopSide_GradentToColorCurve, parameterName=TSCC_Remove_Customization Mask, sortPriority=-10}
  n76 = BlendMaterialAttributes(
   A=@n3
   B=SetMaterialAttributes!(
    MaterialAttributes=@n4.MaterialAttributes
    BaseColor=LinearInterpolate(A=@n4.BaseColor, B=@n68, Alpha=@n71)
    Specular=LinearInterpolate(A=@n4.Specular, B=@n72, Alpha=@n73)
    Metallic
   ) {attributeSetTypes [#2, #3, #4]}
   Alpha=Saturate(None=Add(A=@n74, B=@n75))
  )
  n77 = StaticSwitchParameter(True=@n76, False=@n3) {desc="use mask material to remap gradent to ColorCurves Difuse.  ", group=Customization, parameterName=TopSide_GradentToColorCurve}
  FunctionOutput(None=@n77)
  StaticSwitchParameter(True=@n76, False=Constant()) {parameterName=324dfsgsdg}
 ]}
