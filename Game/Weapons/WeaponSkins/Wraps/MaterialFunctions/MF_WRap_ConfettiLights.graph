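# Weapon-wrap material function adding confetti and party lights, with per-state tuning selected by isState (weapon, vehicle, mech, character).
# Two bounds-projected confetti layers drift and fade from Carnaval trigger-time parameters, driving base color, specular and roughness.
# Emissive sums two sine-swept, hue-shifted spotlights, masked by normals and curvature, with two scrolling disco layers given camera-vector depth offset.
# The Wrap_ConfettiLights static switch blends the result over incoming attributes through customization masks.
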MaterialFunction /Game/Weapons/WeaponSkins/Wraps/MaterialFunctions/MF_WRap_ConfettiLights
graph {nodes [
  n1 = TextureSample() {samplerType=SAMPLERTYPE_Normal, texture=/Game/Weapons/FORT_Rifles/T_SCAR_N}
  n2 = GetMaterialAttributes(None=FunctionInput(Preview=SetMaterialAttributes!(MaterialAttributes, Normal=@n1.RGB) {attributeSetTypes [#1]}) {inputName=Material, inputType=FunctionInput_MaterialAttributes, sortPriority=-20})
  n3 = VectorParameter() {group=Wrap_ConfittiLights, parameterName="Confl_BackgroundColor A=Roughness", defaultValue=<rgba 0 0 0 1>}
  n4 = TextureObjectParameter() {group=Wrap_ConfittiLights, parameterName=Confl_Confetti_texture, texture=&T_BlacksRGB}
  n5 = Subtract(A=Time(), B=ScalarParameter() {defaultValue=-9999999, group=Wrap_Carnaval, parameterName=CanB_StarsB_Main_Time})
  n6 = ScalarParameter() {defaultValue=2, group=Wrap_ConfittiLights, parameterName=Confl_ConfitteSpeed__Duration}
  n7 = Divide(A=@n5, B=@n6)
  n8 = ScalarParameter() {defaultValue=0.5, group=Wrap_ConfittiLights, parameterName=Confl_ConfitteSpeed__Falloff}
  n9 = Power(Base=OneMinus(None=Saturate(None=@n7)), Exp=@n8)
  n10 = VectorParameter() {group=Wrap_ConfittiLights, parameterName=Confl_ConfitteSpeedB_Vehicle}
  n11 = VectorParameter() {group=Wrap_ConfittiLights, parameterName=Confl_ConfitteSpeedB_Vehicle_Side}
  n12 = FunctionInput(Preview=Constant()) {inputName=isState, inputType=FunctionInput_Vector4, sortPriority=5}
  n13 = MaterialFunctionCall(Color=@n12) {materialFunction=&SplitComponents}
  n14 = LinearInterpolate(
   A=LinearInterpolate(
    A=VectorParameter() {group=Wrap_ConfittiLights, parameterName=Confl_ConfitteSpeedB_Weapon}.RGB
    B=LinearInterpolate(A=@n10.RGB, B=@n11.RGB, Alpha=@n13.G)
    Alpha=@n13.R
   )
   B=VectorParameter() {group=Wrap_ConfittiLights, parameterName=Confl_ConfitteSpeedB_Mech}.RGB
   Alpha=@n13.B
  )
  n15 = ScalarParameter() {defaultValue=1, group=Wrap_ConfittiLights, parameterName=Confl_ConfettiB_Scale_Vehicle}
  n16 = ScalarParameter() {defaultValue=1, group=Wrap_ConfittiLights, parameterName=Confl_ConfettiB_Scale_VehicleSide}
  n17 = MaterialFunctionCall(Color=@n12) {materialFunction=&SplitComponents}
  n18 = LinearInterpolate(
   A=LinearInterpolate(
    A=ScalarParameter() {defaultValue=1, group=Wrap_ConfittiLights, parameterName=Confl_ConfettiB_Scale_Weapon}
    B=LinearInterpolate(A=@n15, B=@n16, Alpha=@n17.G)
    Alpha=@n17.R
   )
   B=ScalarParameter() {defaultValue=1, group=Wrap_ConfittiLights, parameterName=Confl_ConfettiB_Scale_Mech}
   Alpha=@n17.B
  )
  n19 = VertexInterpolator(VS=PreSkinnedPosition())
  n20 = PreSkinnedLocalBounds()
  n21 = FunctionInput(Preview=Divide(A=Subtract(A=@n19.PS, B=@n20.Min), B=Subtract(A=@n20.Max, B=@n20.Max))) {bUsePreviewValueAsDefault=true, inputName=Calcated_Bounds, sortPriority=11}
  n22 = MaterialFunctionCall(Color=@n12) {materialFunction=&SplitComponents}
  n23 = Constant3Vector() {constant=<rgba 0 0 0 1>}
  n24 = Constant3Vector() {desc=must activate TS_Vehicle_Side_01 UVs, constant=<rgba 0 0 0 1>}
  n25 = MaterialFunctionCall(Color=@n12) {materialFunction=&SplitComponents}
  n26 = LinearInterpolate(
   A=LinearInterpolate(A=Constant3Vector() {constant=<rgba 0 0 -0.25 1>}, B=LinearInterpolate(A=@n23, B=@n24, Alpha=@n25.G), Alpha=@n25.R)
   B=Constant3Vector() {constant=<rgba 0.25 0.25 0 1>}
   Alpha=@n25.B
  )
  n27 = FunctionInput(Preview=VertexInterpolator(VS=PreSkinnedNormal()).PS) {bUsePreviewValueAsDefault=true, inputName=Pre-Skinned Local Normal, sortPriority=10}
  n28 = ScalarParameter() {group=Wrap_ConfittiLights, parameterName=Confl_Blend_Contrast, sortPriority=10}
  n29 = ScalarParameter() {defaultValue=1, group=Wrap_ConfittiLights, parameterName=Confl_Fill_alternate_Axis_Opacity, sortPriority=1}
  n30 = VectorParameter() {group=Wrap_ConfittiLights, parameterName="Confl_Fill_alternate_Axis_Color_A=Roughness", sortPriority=1}
  n31 = AppendVector(A=@n30.RGB, B=@n30.RGB)
  n32 = MaterialFunctionCall!(
   IsCharacter
   IsStatsB
   Texture=@n4
   Rotation_Axis=@n26
   Caculated_Bounds=Add(A=Multiply(A=@n9, B=@n14), B=Multiply(A=@n21, B=@n18))
   Pre_Skinned_Normals=@n27
   Blend_Contrast=@n28
   Fill_alternate_Axis_Opacity=@n29
   "Fill_alternate_Axis_Color_A=Roughness"=@n31
   Is_Vehicle=@n22.R
   Is_Mech=@n22.R
   Enable_Fill_Z_Weapon
   Fill_Z_Weapon_opacity
   Fill_Z_Weapon_Color
  ) {materialFunction=&/MF_TopsSideProjection_Base}
  n33 = VectorParameter() {group=Wrap_ConfittiLights, parameterName=Confl_Confetti_Color_B, defaultValue=<rgba 0.6673 0.6823 0.6664 1>}
  n34 = Multiply(A=@n32.RGB, B=@n33.RGB)
  n35 = Add(A=@n6) {constB=0.2}
  n36 = Divide(A=@n5, B=@n35)
  n37 = Power(Base=OneMinus(None=Saturate(None=@n36)), Exp=Multiply(A=@n8) {constB=0.5})
  n38 = Multiply(A=Saturate(None=Multiply(A=@n37) {constB=2}), B=@n32.Alpha)
  n39 = Subtract(A=Time(), B=ScalarParameter() {defaultValue=-9999999, group=Wrap_Carnaval, parameterName=CanB_StarsA_Main_Time})
  n40 = Divide(A=@n39, B=@n6)
  n41 = Power(Base=OneMinus(None=Saturate(None=@n40)), Exp=@n8)
  n42 = VectorParameter() {group=Wrap_ConfittiLights, parameterName=Confl_ConfitteSpeed_Vehicle}
  n43 = VectorParameter() {group=Wrap_ConfittiLights, parameterName=Confl_ConfitteSpeed_Vehicle_Side}
  n44 = MaterialFunctionCall(Color=@n12) {materialFunction=&SplitComponents}
  n45 = LinearInterpolate(
   A=LinearInterpolate(
    A=VectorParameter() {group=Wrap_ConfittiLights, parameterName=Confl_ConfitteSpeed_Weapon}.RGB
    B=LinearInterpolate(A=@n42.RGB, B=@n43.RGB, Alpha=@n44.G)
    Alpha=@n44.R
   )
   B=VectorParameter() {group=Wrap_ConfittiLights, parameterName=Confl_ConfitteSpeed_Mech}.RGB
   Alpha=@n44.B
  )
  n46 = ScalarParameter() {defaultValue=1, group=Wrap_ConfittiLights, parameterName=Confl_ConfettiA_Scale_Vehicle}
  n47 = ScalarParameter() {defaultValue=1, group=Wrap_ConfittiLights, parameterName=Confl_ConfettiA_Scale_VehicleSide}
  n48 = MaterialFunctionCall(Color=@n12) {materialFunction=&SplitComponents}
  n49 = LinearInterpolate(
   A=LinearInterpolate(
    A=ScalarParameter() {defaultValue=1, group=Wrap_ConfittiLights, parameterName=Confl_ConfettiA_Scale_Weapon}
    B=LinearInterpolate(A=@n46, B=@n47, Alpha=@n48.G)
    Alpha=@n48.R
   )
   B=ScalarParameter() {defaultValue=1, group=Wrap_ConfittiLights, parameterName=Confl_ConfettiA_Scale_Mech}
   Alpha=@n48.B
  )
  n50 = MaterialFunctionCall(
   Texture=@n4
   Rotation_Axis=@n26
   Caculated_Bounds=Add(A=Multiply(A=@n41, B=@n45), B=Multiply(A=@n21, B=@n49))
   Pre_Skinned_Normals=@n27
   Blend_Contrast=@n28
   Fill_alternate_Axis_Opacity=@n29
   "Fill_alternate_Axis_Color_A=Roughness"=@n31
   Is_Vehicle=@n22.R
   Is_Mech=@n22.R
  ) {materialFunction=&MF_TopsSideProjection_Base}
  n51 = VectorParameter() {group=Wrap_ConfittiLights, parameterName=Confl_Confetti_Color_A, defaultValue=<rgba 1 0.7977 0 1>}
  n52 = Divide(A=@n39, B=@n35)
  n53 = Power(Base=OneMinus(None=Saturate(None=@n52)), Exp=Multiply(A=@n8) {constB=0.5})
  n54 = Multiply(A=Saturate(None=Multiply(A=@n53) {constB=2}), B=@n50.Alpha)
  n55 = Saturate(None=Add(A=@n54, B=@n38))
  n56 = MaterialFunctionCall(Color=Abs(None=@n27)) {materialFunction=&SplitComponents}
  n57 = MaterialFunctionCall(Color=@n12) {materialFunction=&SplitComponents}
  n58 = LinearInterpolate(A=@n56.R, B=LinearInterpolate(A=@n56.B, B=@n56.B, Alpha=@n57.G), Alpha=@n57.R)
  n59 = FunctionInput(Preview=TextureSample() {samplerType=SAMPLERTYPE_Masks, texture=/Game/Weapons/FORT_Rifles/Textures/T_Scar_CM}.G) {bUsePreviewValueAsDefault=true, inputName=Curavture, inputType=FunctionInput_Scalar, sortPriority=24}
  n60 = Multiply(
   A=MaterialFunctionCall(
    In=LinearInterpolate(A=@n58, B=@n56.G, Alpha=@n57.B)
    Contrast=ScalarParameter() {defaultValue=1, group=Wrap_ConfittiLights, parameterName=Confl_Blend_Contrast_Spotlight, sortPriority=10}
   ) {materialFunction=&CheapContrast}.Result
   B=Saturate(None=Multiply(A=@n59) {constB=1.5})
  )
  n61 = Multiply(A=Time(), B=ScalarParameter() {group=Wrap_ConfittiLights, parameterName=Confl_SpotLightA_HueShift_Speed})
  n62 = VectorParameter() {group=Wrap_ConfittiLights, parameterName=Confl_SpotLightA_Color_Outer, defaultValue=<rgba 0.0781 0 0.0012 1>}
  n63 = VectorParameter() {group=Wrap_ConfittiLights, parameterName=Confl_SpotLightA_Color_Inner, defaultValue=<rgba 1 0 0.0158 1>}
  n64 = VectorParameter() {group=Wrap_ConfittiLights, parameterName=Confl_SpotLightA_Offset_Weapon, defaultValue=<rgba -0.5 1 0 0>}
  n65 = VectorParameter() {group=Wrap_ConfittiLights, parameterName=Confl_SpotLightA_Offset_Vehicle_Side, defaultValue=<rgba 0 1 0 0>}
  n66 = MaterialFunctionCall(Color=@n12) {materialFunction=&SplitComponents}
  n67 = LinearInterpolate(A=VectorParameter() {group=Wrap_ConfittiLights, parameterName=Confl_SpotLightA_Offset_Vehicle}.RGB, B=@n65.RGB, Alpha=@n66.G)
  n68 = ComponentMask(None=LinearInterpolate(
    A=LinearInterpolate(A=@n64.RGB, B=@n67, Alpha=@n66.R)
    B=VectorParameter() {group=Wrap_ConfittiLights, parameterName=Confl_SpotLightA_Offset_Mech, defaultValue=<rgba 0 1 0 0>}.RGB
    Alpha=@n66.B
   )) {g=true, r=true}
  n69 = MaterialFunctionCall(Color=@n12) {materialFunction=&SplitComponents}
  n70 = MaterialFunctionCall!(
   For_Safe_Zones
   Normal_Map
   Caculated_Bounds=@n21
   Pre_Skinned_Normals=@n27
   Is Vehicle=@n69.R
   Is VehicleSide=@n69.R
   Is Mech=@n69.R
   Vehicle_Rotation
   UV_Rotation Bypass
   Mech_Rotation
   Uvs_Offset
   Offset-after rotation
   Uvs_TileRate
  ) {materialFunction=&/MF_Projection_Side_Normal}
  n71 = Time()
  n72 = ScalarParameter() {defaultValue=0.1, group=Wrap_ConfittiLights, parameterName=Confl_SpotLightA_Rotation_Speed}
  n73 = Multiply(
   A=Sine(None=Multiply(A=@n71, B=@n72)) {period=0.8}
   B=ScalarParameter() {defaultValue=0.1, group=Wrap_ConfittiLights, parameterName=Confl_SpotLightA_Rotation_Range}
  ) {constB=0.1}
  n74 = ScalarParameter() {group=Wrap_ConfittiLights, parameterName=Confl_SpotLightA_Rotation_Offset_Vehicle}
  n75 = ScalarParameter() {group=Wrap_ConfittiLights, parameterName=Confl_SpotLightA_Rotation_Offset_VehicleSide}
  n76 = MaterialFunctionCall(Color=@n12) {materialFunction=&SplitComponents}
  n77 = LinearInterpolate(
   A=LinearInterpolate(
    A=ScalarParameter() {group=Wrap_ConfittiLights, parameterName=Confl_SpotLightA_Rotation_Offset_Weapon}
    B=LinearInterpolate(A=@n74, B=@n75, Alpha=@n76.G)
    Alpha=@n76.R
   )
   B=ScalarParameter() {group=Wrap_ConfittiLights, parameterName=Confl_SpotLightA_Rotation_Offset_Mech}
   Alpha=@n76.B
  )
  n78 = TextureObjectParameter() {group=Wrap_ConfittiLights, parameterName=Confl_SpotLight_Text, texture=&T_BlacksRGB}
  n79 = TextureSample(
   UVs=MaterialFunctionCall(
    UVs=Add(A=@n70.UVs No Offsets/Tiling, B=@n68)
    Rotation Center=Constant2Vector() {g=1}
    "Rotation Angle (0-1)"=Add(A=@n73, B=@n77) {constB=-1}
   ) {materialFunction=&CustomRotator}.Rotated Values
   Tex=@n78
  )
  n80 = LinearInterpolate(A=@n62.RGB, B=@n63.RGB, Alpha=@n79.RGB)
  n81 = Multiply(A=Time(), B=ScalarParameter() {group=Wrap_ConfittiLights, parameterName=Confl_SpotLightB_HueShift_Speed})
  n82 = VectorParameter() {group=Wrap_ConfittiLights, parameterName=Confl_SpotLightB_Color_Outer, defaultValue=<rgba 0.0781 0 0.0012 1>}
  n83 = VectorParameter() {group=Wrap_ConfittiLights, parameterName=Confl_SpotLightB_Color_Inner, defaultValue=<rgba 1 0 0.0158 1>}
  n84 = VectorParameter() {group=Wrap_ConfittiLights, parameterName=Confl_SpotLightB_Offset_Weapon, defaultValue=<rgba -0.5 1 0 0>}
  n85 = VectorParameter() {group=Wrap_ConfittiLights, parameterName=Confl_SpotLightB_Offset_Vehicle_Side, defaultValue=<rgba 0 1 0 0>}
  n86 = MaterialFunctionCall(Color=@n12) {materialFunction=&SplitComponents}
  n87 = LinearInterpolate(A=VectorParameter() {group=Wrap_ConfittiLights, parameterName=Confl_SpotLightB_Offset_Vehicle}.RGB, B=@n85.RGB, Alpha=@n86.G)
  n88 = ComponentMask(None=LinearInterpolate(
    A=LinearInterpolate(A=@n84.RGB, B=@n87, Alpha=@n86.R)
    B=VectorParameter() {group=Wrap_ConfittiLights, parameterName=Confl_SpotLightB_Offset_Mech, defaultValue=<rgba 0 1 0 0>}.RGB
    Alpha=@n86.B
   )) {g=true, r=true}
  n89 = Multiply(A=Time(), B=ScalarParameter() {defaultValue=0.1, group=Wrap_ConfittiLights, parameterName=Confl_SpotLightB_Rotation_Speed})
  n90 = Multiply(
   A=Sine(None=Add(A=@n89) {constB=0.33}) {period=0.8}
   B=ScalarParameter() {defaultValue=0.1, group=Wrap_ConfittiLights, parameterName=Confl_SpotLightB_Rotation_Range}
  ) {constB=0.1}
  n91 = ScalarParameter() {group=Wrap_ConfittiLights, parameterName=Confl_SpotLightB_Rotation_Offset_Vehicle}
  n92 = ScalarParameter() {group=Wrap_ConfittiLights, parameterName=Confl_SpotLightB_Rotation_Offset_VehicleSide}
  n93 = MaterialFunctionCall(Color=@n12) {materialFunction=&SplitComponents}
  n94 = LinearInterpolate(
   A=LinearInterpolate(
    A=ScalarParameter() {group=Wrap_ConfittiLights, parameterName=Confl_SpotLightB_Rotation_Offset_Weapon}
    B=LinearInterpolate(A=@n91, B=@n92, Alpha=@n93.G)
    Alpha=@n93.R
   )
   B=ScalarParameter() {group=Wrap_ConfittiLights, parameterName=Confl_SpotLightB_Rotation_Offset_Mech}
   Alpha=@n93.B
  )
  n95 = TextureSample(
   UVs=MaterialFunctionCall(
    UVs=Add(A=@n70.UVs No Offsets/Tiling, B=@n88)
    Rotation Center=Constant2Vector() {g=1}
    "Rotation Angle (0-1)"=Add(A=@n90, B=@n94) {constB=-1}
   ) {materialFunction=&CustomRotator}.Rotated Values
   Tex=@n78
  )
  n96 = LinearInterpolate(A=@n82.RGB, B=@n83.RGB, Alpha=@n95.RGB)
  n97 = Add(
   A=Multiply(
    A=MaterialFunctionCall(Hue Shift Percentage=@n61, Texture=@n80) {materialFunction=&HueShift}.Result
    B=@n79.RGB
   )
   B=Multiply(
    A=MaterialFunctionCall(Hue Shift Percentage=@n81, Texture=@n96) {materialFunction=&HueShift}.Result
    B=@n95.RGB
   )
  )
  n98 = MaterialFunctionCall(Hue Shift Percentage=@n61, Texture=@n63.RGB) {materialFunction=&HueShift}
  n99 = Transform(None=CameraVectorWS()) {transformSourceType=TRANSFORMSOURCE_World, transformType=TRANSFORM_Local}
  n100 = MaterialFunctionCall(Float4=@n12) {materialFunction=&BreakOutFloat4Components}
  n101 = LinearInterpolate(
   A=LinearInterpolate(
    A=ScalarParameter() {group=Wrap_ConfittiLights, parameterName=Confl_Disco_Depth_Amount_Weapon}
    B=ScalarParameter() {group=Wrap_ConfittiLights, parameterName=Confl_Disco_Depth_Amount_Vehicle}
    Alpha=@n100.R
   )
   B=ScalarParameter() {group=Wrap_ConfittiLights, parameterName=Confl_Disco_Depth_Amount_Mech}
   Alpha=@n100.B
  )
  n102 = MaterialFunctionCall!(
   Normal_Map
   Caculated_Bounds=Add(A=@n21, B=Multiply(A=@n99, B=@n101))
   Pre_Skinned_Normals=@n27
   Skew Amount=LinearInterpolate(
    A=ScalarParameter() {defaultValue=0.5, group=Wrap_ConfittiLights, parameterName=Confl_Disco_Skew Amount}
    B=ScalarParameter() {defaultValue=0.5, group=Wrap_ConfittiLights, parameterName=Confl_Disco_Skew Amount Character}
    Alpha=@n100.A
   )
   Is Vehicle=@n86.R
   Uvs_Offset
   Uvs_TileRate
   Alternate Axies contrast
  ) {materialFunction=&/MF_Projection_Skew_Normal}
  n103 = MaterialFunctionCall(UVs=@n102.UVs No Offsets/Tiling, "Rotation Angle (0-1)"=ScalarParameter() {group=Wrap_ConfittiLights, parameterName=Confl_Disco_Rotation}) {materialFunction=&CustomRotator}
  n104 = VectorParameter() {group=Wrap_ConfittiLights, parameterName=Confl_Disco_A_Speed_Weapon, defaultValue=<rgba 0 0 0 1>}
  n105 = VectorParameter() {group=Wrap_ConfittiLights, parameterName=Confl_Disco_A_Speed_Vehicle, defaultValue=<rgba 0 0 0 1>}
  n106 = Multiply(A=Time(), B=ComponentMask(None=LinearInterpolate(A=@n104.RGB, B=@n105.RGB, Alpha=@n100.R)) {g=true, r=true})
  n107 = VectorParameter() {group=Wrap_ConfittiLights, parameterName=Confl_Disco_A_TileRate_Weapon, defaultValue=<rgba 1 2 0 0.5>}
  n108 = VectorParameter() {group=Wrap_ConfittiLights, parameterName=Confl_Disco_A_TileRate_Vehicle, defaultValue=<rgba 1 2 0 1>}
  n109 = VectorParameter() {group=Wrap_ConfittiLights, parameterName=Confl_Disco_A_TileRate_Character, defaultValue=<rgba 1 2 0 1>}
  n110 = LinearInterpolate(
   A=LinearInterpolate(A=Multiply(A=@n107.RGB, B=@n107.RGB), B=Multiply(A=@n108.RGB, B=@n108.RGB), Alpha=@n100.R)
   B=Multiply(A=@n109.RGB, B=@n109.RGB)
   Alpha=@n100.A
  )
  n111 = TextureObjectParameter() {group=Wrap_ConfittiLights, parameterName=Confl_Disco__Tex, texture=&T_BlacksRGB}
  n112 = TextureSample(UVs=Multiply(A=Add(A=@n103.Rotated Values, B=@n106), B=ComponentMask(None=@n110) {g=true, r=true}), Tex=@n111)
  n113 = MaterialFunctionCall(Hue Shift Percentage=@n81, Texture=@n83.RGB) {materialFunction=&HueShift}
  n114 = VectorParameter() {group=Wrap_ConfittiLights, parameterName=Confl_Disco_B_Speed_Weapon, defaultValue=<rgba 0 0 0 1>}
  n115 = VectorParameter() {group=Wrap_ConfittiLights, parameterName=Confl_Disco_B_Speed_Vehicle, defaultValue=<rgba 0 0 0 1>}
  n116 = Multiply(A=Time(), B=ComponentMask(None=LinearInterpolate(A=@n114.RGB, B=@n115.RGB, Alpha=@n100.R)) {g=true, r=true})
  n117 = VectorParameter() {group=Wrap_ConfittiLights, parameterName=Confl_Disco_B_TileRate_Weapon, defaultValue=<rgba 1 2 0 0.5>}
  n118 = VectorParameter() {group=Wrap_ConfittiLights, parameterName=Confl_Disco_B_TileRate_Vehicle, defaultValue=<rgba 1 2 0 1>}
  n119 = VectorParameter() {group=Wrap_ConfittiLights, parameterName=Confl_Disco_B_TileRate_Character, defaultValue=<rgba 1 2 0 1>}
  n120 = LinearInterpolate(
   A=LinearInterpolate(A=Multiply(A=@n117.RGB, B=@n117.RGB), B=Multiply(A=@n118.RGB, B=@n118.RGB), Alpha=@n100.R)
   B=Multiply(A=@n119.RGB, B=@n119.RGB)
   Alpha=@n100.A
  )
  n121 = TextureSample(UVs=Multiply(A=Add(A=@n103.Rotated Values, B=@n116), B=ComponentMask(None=@n120) {g=true, r=true}), Tex=@n111)
  n122 = Add(
   A=Multiply(
    A=Multiply(A=@n98.Result, B=@n112.B)
    B=ScalarParameter() {defaultValue=1, group=Wrap_ConfittiLights, parameterName=Confl_Disco_A_Brightness}
   )
   B=Multiply(
    A=ScalarParameter() {defaultValue=1, group=Wrap_ConfittiLights, parameterName=Confl_Disco_B_Brightness}
    B=Multiply(A=@n113.Result, B=@n121.B)
   )
  )
  n123 = SetMaterialAttributes!(
   MaterialAttributes=@n2.MaterialAttributes
   BaseColor=LinearInterpolate(A=LinearInterpolate(A=@n3.RGB, B=@n34, Alpha=@n38), B=Multiply(A=@n50.RGB, B=@n51.RGB), Alpha=@n54)
   Metallic=Constant()
   Specular=LinearInterpolate(
    A=ScalarParameter() {defaultValue=0.5, group=Wrap_ConfittiLights, parameterName=Confl_Background Specular, sortPriority=1}
    B=ScalarParameter() {defaultValue=0.5, group=Wrap_ConfittiLights, parameterName=Confl_Confetti Specular, sortPriority=1}
    Alpha=@n55
   )
   Roughness=LinearInterpolate(A=@n3.A, B=ScalarParameter() {defaultValue=1, group=Wrap_ConfittiLights, parameterName=Confl_Confetti Roughness}, Alpha=@n55)
   EmissiveColor=Add(A=Multiply(A=@n60, B=@n97), B=Multiply(A=@n102.Masking Alternat Axies, B=@n122))
  ) {attributeSetTypes [
    #2
    #3
    #4
    #5
    #6
   ]}
  n124 = FunctionInput() {bUsePreviewValueAsDefault=true, inputName="Customization Mask (_CM B)", inputType=FunctionInput_Scalar, sortPriority=24}
  n125 = ScalarParameter() {group=Wrap_ConfittiLights, parameterName=Confl_Remove_Customization Mask, sortPriority=-10}
  n126 = NamedRerouteUsage()
  n127 = NamedRerouteUsage()
  n128 = LinearInterpolate(
   A=Saturate(None=Add(A=@n124, B=@n125))
   B=FunctionInput() {bUsePreviewValueAsDefault=true, inputName=Customization Mask B, inputType=FunctionInput_Scalar, sortPriority=24}
   Alpha=Saturate(None=Subtract(A=@n126, B=@n127))
  )
  FunctionOutput(None=StaticSwitchParameter(True=BlendMaterialAttributes(A=@n2.MaterialAttributes, B=@n123, Alpha=@n128) {vertexAttributeBlendType=UseA}, False=@n2.MaterialAttributes) {group=Customization_Wrap, parameterName=Wrap_ConfettiLights})
  NamedRerouteDeclaration(None=MaterialFunctionCall(Float4=@n12) {materialFunction=&BreakOutFloat4Components}.A) {name=IsCharacter, nodeColor=<rgba 0.9898 1 0 1>}
  NamedRerouteDeclaration(None=FunctionInput() {bUsePreviewValueAsDefault=true, inputName=isCharAccent, inputType=FunctionInput_Scalar, sortPriority=78}) {name=isCharAccent, nodeColor=<rgba 0 0.9253 1 1>}
 ]}
