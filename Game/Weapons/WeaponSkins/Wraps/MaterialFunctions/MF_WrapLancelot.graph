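# Material function for the Lancelot weapon wrap, layering the skin over an incoming material behind the Wrap_Lancelot static switch.
# Textures use side projection from pre-skinned bounds and normals, with offset, scale and rotation selected per item type.
# The skin combines a fresnel metal LUT, pulsing time-reactive blade glow, and procedural eye and corruption glows from scrolling noise through curve-atlas ramps.
# Primary and secondary input masks place the skin, with a separate accent material on the secondary.
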MaterialFunction /Game/Weapons/WeaponSkins/Wraps/MaterialFunctions/MF_WrapLancelot
graph {nodes [
  n1 = TextureSample(UVs=NamedRerouteUsage(), Tex=NamedRerouteUsage())
  n2 = VectorParameter() {group=Wrap_Lancelot, parameterName=Lance_Blade_Color, defaultValue=<rgba 0.004 0.9301 0.7991 1>}
  n3 = NamedRerouteUsage()
  n4 = Add(A=Multiply(A=Multiply(A=@n1.R, B=@n2.RGB), B=Add(A=@n3)), B=Multiply(A=@n2.RGB, B=Multiply(A=@n3, B=@n1.B) {constA=0.5}))
  n5 = Max(A=NamedRerouteUsage(), B=@n3)
  n6 = Step(X=@n1.R) {constX=0.9, constY=0.2}
  n7 = OneMinus(None=@n6)
  n8 = Multiply(A=ComponentMask(None=NamedRerouteUsage()) {a=true}, B=ComponentMask(None=NamedRerouteUsage()) {g=true})
  n9 = Frac(None=Multiply(A=@n8) {constB=0.8})
  n10 = CurveAtlasRowParameter(CurveTime=@n9) {
   atlas=/Game/Weapons/WeaponSkins/Wraps/Textures/CruveAtlas_Wraps_01
   curve=/Game/Weapons/WeaponSkins/Wraps/Textures/0_ColorCurves/CC_Lance_Eyes
   defaultValue=77
   group=Wrap_Lancelot
   parameterName=Lance_CC_Eyes
  }
  n11 = Multiply(
   A=MaterialFunctionCall!(Index=ScalarParameter() {defaultValue=77, group=Wrap_Lancelot, parameterName="Lance_CC_Eyes  Index"}, CCinput=@n10, CCinput_Alpha, UV=@n9) {materialFunction=&/MF_Global_CC_Tex}.3V
   B=@n8
  )
  n12 = OneMinus(None=Add(A=NamedRerouteUsage(), B=NamedRerouteUsage()))
  n13 = Multiply(A=Add(A=Multiply(A=@n4, B=@n5), B=Multiply(A=@n7, B=@n11)), B=@n12)
  n14 = NamedRerouteUsage()
  n15 = NamedRerouteUsage()
  Add(A=@n13, B=Add(A=@n14, B=@n15))
  Add(A=Multiply(A=Time(), B=ScalarParameter() {defaultValue=0.1, group=Wrap_Lancelot, parameterName=Lance_CC_Eyes_Speed}), B=@n8)
  n16 = NamedRerouteUsage()
  n17 = ComponentMask(None=@n16) {g=true}
  n18 = NamedRerouteUsage()
  n19 = Multiply(A=Multiply(A=OneMinus(None=@n17), B=ComponentMask(None=@n18) {a=true})) {constB=2}
  Add(A=@n19)
  n20 = NamedRerouteUsage()
  n21 = VectorParameter() {group=Wrap_Lancelot, parameterName=Lance_BC, sortPriority=10, defaultValue=<rgba 0.001 0.001 0.001 1>}
  n22 = NamedRerouteUsage()
  n23 = LinearInterpolate(
   A=ScalarParameter() {defaultValue=0.15, group=Wrap_Lancelot, parameterName=Lance_Specular, sortPriority=10}
   B=ScalarParameter() {defaultValue=0.5, group=Wrap_Lancelot, parameterName=Lance_Specular_Curroption, sortPriority=10}
   Alpha=@n22
  )
  n24 = SetMaterialAttributes!(
   MaterialAttributes=@n20
   BaseColor=@n21.RGB
   Roughness=ScalarParameter() {defaultValue=0.45, group=Wrap_Lancelot, parameterName=Lance_Roughness, sortPriority=10}
   Specular=@n23
   Metallic
   Normal
   EmissiveColor
  ) {attributeSetTypes [
    #1
    #2
    #3
    #4
    #5
    #6
   ]}
  n25 = Clamp(None=DotProduct(A=CameraVectorWS(), B=PixelNormalWS()))
  n26 = CurveAtlasRowParameter(CurveTime=@n25) {
   atlas=&CruveAtlas_Wraps_01
   curve=/Game/Weapons/WeaponSkins/Wraps/Textures/0_ColorCurves/CC_MetalicLUT
   defaultValue=76
   group=Wrap_Lancelot
   parameterName=Lance_LUT_Metal_CC_Fresnel
  }
  n27 = MaterialFunctionCall(
   Index=ScalarParameter() {defaultValue=76, group=Wrap_Lancelot, parameterName="Lance_LUT_Metal_CC_Fresnel  Index"}
   CCinput=@n26
   CCinput_Alpha=@n26
   UV=@n25
  ) {materialFunction=&MF_Global_CC_Tex}
  n28 = MaterialFunctionCall(Float4=NamedRerouteUsage()) {materialFunction=&BreakOutFloat4Components}
  n29 = MaterialFunctionCall!(
   For_Safe_Zones
   Normal_Map=NamedRerouteUsage()
   Caculated_Bounds=NamedRerouteUsage()
   Pre_Skinned_Normals=NamedRerouteUsage()
   Is Vehicle=@n28.R
   Is VehicleSide=@n28.R
   Is Mech=@n28.R
   Vehicle_Rotation
   UV_Rotation Bypass=NamedRerouteUsage()
   Mech_Rotation
   Uvs_Offset=NamedRerouteUsage()
   Offset-after rotation
   Uvs_TileRate=NamedRerouteUsage()
  ) {materialFunction=&/MF_Projection_Side_Normal}
  n30 = TextureSample(UVs=NamedRerouteDeclaration(None=@n29.UVs) {name=Uvs, nodeColor=<rgba 1 0 0.1273 1>}, Tex=NamedRerouteUsage())
  n31 = TextureSample(UVs=NamedRerouteUsage(), Tex=NamedRerouteUsage())
  n32 = Add(A=@n6, B=@n31.R)
  n33 = LinearInterpolate(
   A=@n30.RGB
   B=Multiply(A=@n30.RGB, B=ChannelMaskParameter(None=@n27.4v) {group=Wrap_Lancelot, parameterName=Lance_LUT_Metal_CC_Channel}.R)
   Alpha=@n32
  )
  n34 = ScalarParameter() {group=Wrap_Lancelot, parameterName=Lance_Roughness_High, sortPriority=10}
  n35 = ScalarParameter() {defaultValue=1, group=Wrap_Lancelot, parameterName=Lance_Roughness_Low, sortPriority=10}
  n36 = SetMaterialAttributes(
   MaterialAttributes=@n20
   BaseColor=Add(A=Multiply(A=@n33, B=@n12), B=@n15)
   Roughness=LinearInterpolate(
    A=LinearInterpolate(A=@n34, B=@n35, Alpha=@n30.A)
    B=ScalarParameter() {defaultValue=0.3, group=Wrap_Lancelot, parameterName=Lance_Roughness_Curroption, sortPriority=10}
    Alpha=@n22
   )
   Specular=@n23
   Metallic=Multiply(A=@n32, B=ScalarParameter() {defaultValue=0.9, group=Wrap_Lancelot, parameterName=Lance_Metalic, sortPriority=10})
   Normal=Add(A=@n29.Raw_Normal, B=NamedRerouteUsage()) {constB=0}
   EmissiveColor=Add(A=@n13, B=@n14)
  ) {attributeSetTypes [
    #1
    #2
    #3
    #4
    #5
    #6
   ]}
  BlendMaterialAttributes(A=@n24, B=@n36, Alpha=NamedRerouteUsage()) {vertexAttributeBlendType=UseA}
  ComponentMask() {g=true, r=true}
  ComponentMask(None=@n16) {r=true}
  n37 = TextureSample() {samplerType=SAMPLERTYPE_Masks, texture=/Game/Weapons/FORT_Rifles/Textures/T_Scar_CM}
  n38 = FunctionInput(Preview=@n37.R) {bUsePreviewValueAsDefault=true, inputName=Mask Primary, inputType=FunctionInput_Scalar, sortPriority=24, previewValue=<1 0 0 1>}
  n39 = FunctionInput(Preview=@n37.B) {bUsePreviewValueAsDefault=true, inputName=Mask Secondary, inputType=FunctionInput_Scalar, sortPriority=24, previewValue=<1 0 0 1>}
  n40 = BlendMaterialAttributes(A=NamedRerouteUsage(), B=NamedRerouteUsage(), Alpha=Saturate(None=Add(A=@n38, B=@n39))) {vertexAttributeBlendType=UseA}
  n41 = NamedRerouteUsage()
  n42 = NamedRerouteDeclaration(None=@n39) {name=Msk_Accent, nodeColor=<rgba 0 1 0.7656 1>}
  n43 = TextureSample() {samplerType=SAMPLERTYPE_Normal, texture=/Game/Weapons/FORT_Rifles/T_SCAR_N}
  n44 = Constant3Vector() {constant=<rgba 1 0 0 0>}
  n45 = GetMaterialAttributes(None=FunctionInput(Preview=SetMaterialAttributes!(MaterialAttributes, Normal=@n43.RGB, BaseColor=@n44) {attributeSetTypes [#5, #1]}) {inputName=Material, inputType=FunctionInput_MaterialAttributes}) {attributeGetTypes [
    #1
    #2
    #5
    #4
   ]}
  FunctionOutput(None=StaticSwitchParameter(
    True=BlendMaterialAttributes(A=@n40, B=@n41, Alpha=@n42) {vertexAttributeBlendType=UseA}
    False=NamedRerouteDeclaration(None=@n45.MaterialAttributes) {name=Root_Material, nodeColor=<rgba 0.7812 1 0 1>}
   ) {group=Customization_Wrap, parameterName=Wrap_Lancelot})
  n46 = VectorParameter() {group=Wrap_Lancelot_Accent, parameterName=Lance_Accent_BaseColor, defaultValue=<rgba 0.004 0.004 0.004 1>}
  n47 = SetMaterialAttributes!(
   MaterialAttributes=NamedRerouteUsage()
   BaseColor=@n46.RGB
   Roughness=ScalarParameter() {defaultValue=0.3, group=Wrap_Lancelot_Accent, parameterName=Lance_Accent_Roughness}
   Specular=ScalarParameter() {defaultValue=0.25, group=Wrap_Lancelot_Accent, parameterName=Lance_Accent_Spec}
   Metallic=ScalarParameter() {defaultValue=0.5, group=Wrap_Lancelot_Accent, parameterName=Lance_Accent_Metalic}
   Normal
   EmissiveColor
  ) {attributeSetTypes [
    #1
    #2
    #3
    #4
    #5
    #6
   ]}
  NamedRerouteDeclaration(None=@n47) {name=Mat_Accent, nodeColor=<rgba 0 0.3906 1 1>}
  n48 = FunctionInput(Preview=@n37.G) {bUsePreviewValueAsDefault=true, inputName=Mask Crevise, inputType=FunctionInput_Scalar, sortPriority=24, previewValue=<1 0 0 1>}
  NamedRerouteDeclaration(None=@n48) {name=Msk_Curvature, nodeColor=<rgba 0 1 0.7422 1>}
  NamedRerouteDeclaration(None=@n31.RGBA) {name=Tex_Uv_Msk2, nodeColor=<rgba 0 0.25 1 1>}
  NamedRerouteDeclaration(None=TextureObjectParameter() {group=Wrap_Lancelot, parameterName=Lance__Tex_Msk1, texture=&T_BlacksRGB}) {name=Text_Msk1, nodeColor=<rgba 0 0.1562 1 1>}
  NamedRerouteDeclaration(None=@n1.RGBA) {name=Tex_Uv_Msk1, nodeColor=<rgba 0 0.25 1 1>}
  NamedRerouteDeclaration(None=TextureObjectParameter() {group=Wrap_Lancelot, parameterName=Lance__Tex_Msk2, texture=&T_BlacksRGB}) {name=Text_Msk2, nodeColor=<rgba 0 0.1562 1 1>}
  n49 = ScalarParameter() {defaultValue=0.25, group=Wrap_Lancelot, parameterName=Lance_Blade_Speed}
  n50 = Constant() {r=0.033}
  n51 = Frac(None=Multiply(A=Time(), B=AppendVector(A=@n49, B=@n50)))
  n52 = NamedRerouteUsage()
  NamedRerouteDeclaration(None=LinearInterpolate(Alpha=TextureSample(UVs=@n51, Tex=@n52).A) {constA=-0.15, constB=2}) {name=Sword_Pulse, nodeColor=<rgba 1 0 0.4922 1>}
  n53 = Time()
  n54 = ComponentMask(None=VectorParameter() {group=Wrap_Lancelot, parameterName=Lance_Eye_Noise_Speed, defaultValue=<rgba 0.0031 0.05 0 1>}.RGB) {g=true, r=true}
  n55 = Multiply(
   A=ScalarParameter() {defaultValue=8, group=Wrap_Lancelot, parameterName=Lance_Eye_Noise_TileRate}
   B=Add(A=NamedRerouteUsage(), B=Multiply(A=@n53, B=@n54))
  )
  n56 = NamedRerouteUsage()
  n57 = ScalarParameter() {defaultValue=0.05, group=Wrap_Lancelot, parameterName=Lance_Eye_Blob_Size_B}
  n58 = NamedRerouteUsage()
  n59 = ComponentMask(None=VectorParameter() {group=Wrap_Lancelot, parameterName=Lance_Eye_Position, defaultValue=<rgba -0.95 -0.493 0 0.021>}.RGB) {g=true, r=true}
  n60 = Add(A=@n58, B=Add(A=@n59, B=Constant2Vector() {g=0.025}))
  n61 = MaterialFunctionCall(Vector 3=@n60, Vector 2=@n60) {materialFunction=/Engine/Functions/Engine_MaterialFunctions02/Utility/VectorLength}
  n62 = Multiply(
   A=Multiply(B=TextureSample(UVs=@n55, Tex=@n56).A) {constA=0.8}
   B=SmoothStep(Min=Multiply(A=@n57) {constB=1.5}, Value=@n61.V2 Length) {constMax=0, constMin=0.035}
  )
  n63 = Multiply(A=SmoothStep(Min=@n57, Value=@n61.V2 Length) {constMax=0, constMin=0.035}) {constB=0.25}
  n64 = MaterialFunctionCall(Vector 2=Add(A=@n58, B=@n59) {constB=-0.5}) {materialFunction=&VectorLength}
  n65 = SmoothStep(Min=ScalarParameter() {defaultValue=0.025, group=Wrap_Lancelot, parameterName=Lance_Eye_Blob_Size_A}, Value=@n64.V2 Length) {constMax=0, constMin=0.021}
  n66 = LinearInterpolate(
   A=Add(A=Add(A=@n62, B=@n63), B=Multiply(A=@n65) {constB=0.75})
   Alpha=MaterialFunctionCall(In=SmoothStep(Value=@n64.V2 Length) {constMax=0, constMin=0.01}, Contrast=Constant() {r=0.5}) {materialFunction=&CheapContrast}.Result
  )
  n67 = CurveAtlasRowParameter(CurveTime=@n66) {
   atlas=&CruveAtlas_Wraps_01
   curve=/Game/Weapons/WeaponSkins/Wraps/Textures/0_ColorCurves/CC_Lance_Eyes_Noise
   defaultValue=78
   group=Wrap_Lancelot
   parameterName=Lance_Eye_Noise_CC
  }
  n68 = ComponentMask(None=MaterialFunctionCall(Index=ScalarParameter() {defaultValue=78, group=Wrap_Lancelot, parameterName="Lance_Eye_Noise_CC  Index"}, CCinput=@n67, UV=@n66) {materialFunction=&MF_Global_CC_Tex}.3V) {g=true}
  n69 = CurveAtlasRowParameter(CurveTime=@n68) {
   atlas=&CruveAtlas_Wraps_01
   curve=/Game/Weapons/WeaponSkins/Wraps/Textures/0_ColorCurves/CC_Lance_Eyes_Glow
   defaultValue=79
   group=Wrap_Lancelot
   parameterName=Lance_CC_Eyes_Noise_Glow
  }
  n70 = MaterialFunctionCall(
   Index=ScalarParameter() {defaultValue=79, group=Wrap_Lancelot, parameterName="Lance_CC_Eyes_Noise_Glow  Index"}
   CCinput=@n69
   CCinput_Alpha=@n69
   UV=@n68
  ) {materialFunction=&MF_Global_CC_Tex}
  NamedRerouteDeclaration(None=@n70.Alpha) {name=Eye_Noise_Msk, nodeColor=<rgba 1 0.9141 0 1>}
  NamedRerouteDeclaration(None=@n36) {name=Mat_Base, nodeColor=<rgba 0 0.3906 1 1>}
  n71 = SmoothStep(Value=@n64.V2 Length) {constMax=0, constMin=0.005}
  n72 = Constant() {r=0.5}
  NamedRerouteDeclaration(None=LinearInterpolate(
    A=Multiply(A=@n70.3V, B=@n70.3V)
    Alpha=MaterialFunctionCall(In=@n71, Contrast=@n72) {materialFunction=&CheapContrast}.Result
   ) {constB=0}) {name=Eye_Noise_Glow, nodeColor=<rgba 1 0.9141 0 1>}
  NamedRerouteDeclaration(None=TextureObjectParameter() {group=Wrap_Lancelot, parameterName=Lance__Tex_BC, texture=&T_BlacksRGB}) {name=Text_BC, nodeColor=<rgba 0 0.1562 1 1>}
  n73 = ScalarParameter() {defaultValue=78, group=Wrap_Lancelot, parameterName="Lance_Eye_Noise_CC  Index"}
  n74 = NamedRerouteUsage()
  n75 = VectorParameter() {group=Wrap_Lancelot, parameterName=Lance_Curption_Noise_Speed, defaultValue=<rgba 0.0031 0.05 0 1>}
  n76 = Add(A=@n74, B=Multiply(A=Time(), B=ComponentMask(None=@n75.RGB) {g=true, r=true}))
  n77 = Multiply(A=@n76, B=ScalarParameter() {defaultValue=10, group=Wrap_Lancelot, parameterName=Lance_Curption_NoiseB_TileRate})
  n78 = NamedRerouteUsage()
  n79 = Add(
   A=NamedRerouteUsage()
   B=ComponentMask(None=VectorParameter() {group=Wrap_Lancelot, parameterName=Lance_Curption_Position_A, defaultValue=<rgba -0.9 -0.5 0 0.021>}.RGB) {g=true, r=true}
  ) {constB=-0.5}
  n80 = Multiply(A=SmoothStep(
    Max=NamedRerouteUsage()
    Value=MaterialFunctionCall(Vector 2=@n79) {materialFunction=&VectorLength}.V2 Length
   ) {constMax=0.2}) {constB=0.6}
  n81 = ScalarParameter() {defaultValue=2, group=Wrap_Lancelot, parameterName=Lance_Curption_Falloff}
  n82 = Add(
   A=Multiply(A=TextureSample(UVs=@n77, Tex=@n78) {samplerSource=SSM_Wrap_WorldGroupSettings}.A) {constB=0.5}
   B=Multiply(A=@n19, B=Power(Base=@n80, Exp=@n81) {constExponent=2.5}) {constA=1}
  )
  n83 = CurveAtlasRowParameter(CurveTime=@n82) {
   atlas=&CruveAtlas_Wraps_01
   curve=&CC_Lance_Eyes_Noise
   defaultValue=78
   group=Wrap_Lancelot
   parameterName=Lance_Eye_Noise_CC
  }
  n84 = Multiply(
   A=NamedRerouteUsage()
   B=ComponentMask(None=MaterialFunctionCall(Index=@n73, CCinput=@n83, UV=@n82) {materialFunction=&MF_Global_CC_Tex}.3V) {r=true}
  )
  NamedRerouteDeclaration(None=@n84) {name=Curption_Msk, nodeColor=<rgba 1 0.9141 0 1>}
  n85 = ScalarParameter() {defaultValue=4, group=Wrap_Lancelot, parameterName=Lance_Curption_Noise_TileRate}
  n86 = Multiply(A=TextureSample(UVs=Multiply(A=@n85, B=@n76), Tex=@n78) {samplerSource=SSM_Wrap_WorldGroupSettings}.A) {constB=1.5}
  n87 = Multiply(A=Time()) {constB=0.5}
  n88 = Frac(None=Add(A=Add(A=@n86, B=@n87), B=ComponentMask(None=@n74) {g=true}))
  n89 = CurveAtlasRowParameter(CurveTime=@n88) {
   atlas=&CruveAtlas_Wraps_01
   curve=/Game/Weapons/WeaponSkins/Wraps/Textures/0_ColorCurves/CC_Lance_Curption_Glow
   defaultValue=80
   group=Wrap_Lancelot
   parameterName=Lance_Curption_Noise_CC_Glow
  }
  n90 = MaterialFunctionCall(
   Index=ScalarParameter() {defaultValue=80, group=Wrap_Lancelot, parameterName="Lance_Curption_Noise_CC_Glow  Index"}
   CCinput=@n89
   CCinput_Alpha=@n89
   UV=@n88
  ) {materialFunction=&MF_Global_CC_Tex}
  NamedRerouteDeclaration(None=Multiply(A=@n84, B=Multiply(A=@n90.3V, B=@n90.3V))) {name=Curption_Glow, nodeColor=<rgba 1 0.9141 0 1>}
  n91 = VectorParameter() {group=Wrap_Lancelot_UV, parameterName=Lance_Offset_Vehicle, defaultValue=<rgba 0 0 0 1>}
  n92 = VectorParameter() {group=Wrap_Lancelot_UV, parameterName=Lance_Offset_VehicleSide, defaultValue=<rgba 0 0 0 1>}
  n93 = MaterialFunctionCall(Float4=NamedRerouteUsage()) {materialFunction=&BreakOutFloat4Components}
  n94 = LinearInterpolate(
   A=LinearInterpolate(
    A=VectorParameter() {group=Wrap_Lancelot_UV, parameterName=Lance_Offset_Weapon, defaultValue=<rgba 0 0 0 1>}.RGB
    B=LinearInterpolate(A=@n91.RGB, B=@n92.RGB, Alpha=@n93.G)
    Alpha=@n93.R
   )
   B=LinearInterpolate(
    A=VectorParameter() {group=Wrap_Lancelot_UV, parameterName=Lance_Offset_Mech, defaultValue=<rgba 0 0 0 1>}.RGB
    B=VectorParameter() {group=Wrap_Lancelot_UV, parameterName=Lance_Offset_Character, defaultValue=<rgba 0 0 0 1>}.RGB
    Alpha=@n93.A
   )
   Alpha=@n93.B
  )
  n95 = VectorParameter() {group=Wrap_Lancelot_UV, parameterName=Lance_Offset_Backpack, defaultValue=<rgba 0 0 0 1>}
  n96 = MaterialFunctionCall(Float2=NamedRerouteUsage()) {materialFunction=&BreakOutFloat2Components}
  NamedRerouteDeclaration(None=LinearInterpolate(
    A=LinearInterpolate(A=@n94, B=@n95.RGB, Alpha=@n96.R)
    B=VectorParameter() {group=Wrap_Lancelot_UV, parameterName=Lance_Offset_Pickaxe, defaultValue=<rgba 0 0 0 1>}.RGB
    Alpha=@n96.G
   )) {name=UVs_Offset, nodeColor=<rgba 0 0.625 1 1>}
  n97 = MaterialFunctionCall(Color=Abs(None=NamedRerouteUsage())) {materialFunction=&SplitComponents}
  n98 = MaterialFunctionCall(Float4=NamedRerouteUsage()) {materialFunction=&BreakOutFloat4Components}
  n99 = LinearInterpolate(A=@n97.R, B=LinearInterpolate(A=@n97.B, B=@n97.B, Alpha=@n98.G), Alpha=@n98.R)
  NamedRerouteDeclaration(None=MaterialFunctionCall(
    In=LinearInterpolate(A=@n99, B=@n97.G, Alpha=@n98.B)
    Contrast=ScalarParameter() {defaultValue=1, group=Wrap_Lancelot, parameterName=Lance_Tri_Blend_Contrast, sortPriority=0}
   ) {materialFunction=&CheapContrast}.Result) {name=Tri_Mask, nodeColor=<rgba 1 0.4688 0 1>}
  n100 = ScalarParameter() {defaultValue=1, group=Wrap_Lancelot_UV, parameterName=Lance_Scale_Vehicle, sortPriority=150}
  n101 = ScalarParameter() {defaultValue=1, group=Wrap_Lancelot_UV, parameterName=Lance_Scale_Vehicle_Side, sortPriority=150}
  n102 = MaterialFunctionCall(Float4=NamedRerouteUsage()) {materialFunction=&BreakOutFloat4Components}
  n103 = LinearInterpolate(
   A=LinearInterpolate(
    A=ScalarParameter() {defaultValue=1, group=Wrap_Lancelot_UV, parameterName=Lance_Scale_Weapon, sortPriority=150}
    B=LinearInterpolate(A=@n100, B=@n101, Alpha=@n102.G)
    Alpha=@n102.R
   )
   B=LinearInterpolate(
    A=ScalarParameter() {defaultValue=1, group=Wrap_Lancelot_UV, parameterName=Lance_Scale_Mech, sortPriority=150}
    B=ScalarParameter() {defaultValue=1, group=Wrap_Lancelot_UV, parameterName=Lance_Scale_Character, sortPriority=150}
    Alpha=@n102.A
   )
   Alpha=@n102.B
  )
  n104 = ScalarParameter() {defaultValue=1, group=Wrap_Lancelot_UV, parameterName=Lance_Scale_Backpack, sortPriority=150}
  n105 = MaterialFunctionCall(Float2=NamedRerouteUsage()) {materialFunction=&BreakOutFloat2Components}
  NamedRerouteDeclaration(None=LinearInterpolate(
    A=LinearInterpolate(A=@n103, B=@n104, Alpha=@n105.R)
    B=ScalarParameter() {defaultValue=1, group=Wrap_Lancelot_UV, parameterName=Lance_Scale_Pickaxe, sortPriority=150}
    Alpha=@n105.G
   )) {name=UVs_Scale, nodeColor=<rgba 0 0.7188 1 1>}
  n106 = ScalarParameter() {group=Wrap_Lancelot_UV, parameterName=Lance_Rotation_Vehicle, sortPriority=150}
  n107 = ScalarParameter() {group=Wrap_Lancelot_UV, parameterName=Lance_Rotation_Vehicle_Side, sortPriority=150}
  n108 = MaterialFunctionCall(Float4=NamedRerouteUsage()) {materialFunction=&BreakOutFloat4Components}
  n109 = LinearInterpolate(
   A=LinearInterpolate(
    A=ScalarParameter() {group=Wrap_Lancelot_UV, parameterName=Lance_Rotation_Weapon, sortPriority=150}
    B=LinearInterpolate(A=@n106, B=@n107, Alpha=@n108.G)
    Alpha=@n108.R
   )
   B=LinearInterpolate(
    A=ScalarParameter() {group=Wrap_Lancelot_UV, parameterName=Lance_Rotation_Mech, sortPriority=150}
    B=ScalarParameter() {group=Wrap_Lancelot_UV, parameterName=Lance_Rotation_Character, sortPriority=150}
    Alpha=@n108.A
   )
   Alpha=@n108.B
  )
  n110 = ScalarParameter() {group=Wrap_Lancelot_UV, parameterName=Lance_Rotation_Backpack, sortPriority=150}
  n111 = MaterialFunctionCall(Float2=NamedRerouteUsage()) {materialFunction=&BreakOutFloat2Components}
  NamedRerouteDeclaration(None=LinearInterpolate(
    A=LinearInterpolate(A=@n109, B=@n110, Alpha=@n111.R)
    B=ScalarParameter() {group=Wrap_Lancelot_UV, parameterName=Lance_Rotation_Pickaxe, sortPriority=150}
    Alpha=@n111.G
   )) {name=UVs_Rotation, nodeColor=<rgba 0 0.7188 1 1>}
  n112 = ScalarParameter() {defaultValue=0.1, group=Wrap_Lancelot_UV, parameterName=Lance_Curption_SizeVehicle, sortPriority=150}
  n113 = ScalarParameter() {defaultValue=0.1, group=Wrap_Lancelot_UV, parameterName=Lance_Curption_SizeVehicle_Side, sortPriority=150}
  n114 = MaterialFunctionCall(Float4=NamedRerouteUsage()) {materialFunction=&BreakOutFloat4Components}
  n115 = LinearInterpolate(
   A=LinearInterpolate(
    A=ScalarParameter() {defaultValue=0.1, group=Wrap_Lancelot_UV, parameterName=Lance_Curption_Size_Weapon, sortPriority=150}
    B=LinearInterpolate(A=@n112, B=@n113, Alpha=@n114.G)
    Alpha=@n114.R
   )
   B=LinearInterpolate(
    A=ScalarParameter() {defaultValue=0.1, group=Wrap_Lancelot_UV, parameterName=Lance_Curption_SizeMech, sortPriority=150}
    B=ScalarParameter() {defaultValue=0.1, group=Wrap_Lancelot_UV, parameterName=Lance_Curption_SizeCharacter, sortPriority=150}
    Alpha=@n114.A
   )
   Alpha=@n114.B
  )
  n116 = ScalarParameter() {defaultValue=0.1, group=Wrap_Lancelot_UV, parameterName=Lance_Curption_SizeBackpack, sortPriority=150}
  n117 = MaterialFunctionCall(Float2=NamedRerouteUsage()) {materialFunction=&BreakOutFloat2Components}
  NamedRerouteDeclaration(None=LinearInterpolate(
    A=LinearInterpolate(A=@n115, B=@n116, Alpha=@n117.R)
    B=ScalarParameter() {defaultValue=0.1, group=Wrap_Lancelot_UV, parameterName=Lance_Curption_SizePickaxe, sortPriority=150}
    Alpha=@n117.G
   )) {name=Curption_Size, nodeColor=<rgba 0 0.7188 1 1>}
  n118 = Time()
  n119 = ScalarParameter() {defaultValue=-9999999, group=Wrap_Reactive, parameterName=TimeFrom-BP----------------, sortPriority=600}
  n120 = Saturate(None=Divide(A=Subtract(A=@n118, B=@n119), B=ScalarParameter() {defaultValue=0.25, group=Wrap_Lancelot_Reactive, parameterName=Lance_Duration}))
  NamedRerouteDeclaration(None=LinearInterpolate(
    A=ScalarParameter() {defaultValue=0.5, group=Wrap_Lancelot_Reactive, parameterName=Lance_Blade__Start}
    B=ScalarParameter() {group=Wrap_Lancelot_Reactive, parameterName=Lance_Blade__End}
    Alpha=@n120
   )) {name=Reactive, nodeColor=<rgba 0 0.2031 1 1>}
  NamedRerouteDeclaration(None=FunctionInput(Preview=Constant4Vector().RGBA) {bUsePreviewValueAsDefault=true, inputName=IsState, inputType=FunctionInput_Vector4, sortPriority=4}) {name=IsState, nodeColor=<rgba 0.1497 1 0.1633 1>}
  NamedRerouteDeclaration(None=FunctionInput(Preview=Constant2Vector()) {bUsePreviewValueAsDefault=true, inputName=IsState_B, inputType=FunctionInput_Vector2, sortPriority=4}) {name=IsState_B, nodeColor=<rgba 0.1497 1 0.1633 1>}
  n121 = PreSkinnedLocalBounds()
  n122 = Subtract(A=VertexInterpolator(VS=PreSkinnedPosition()).PS, B=@n121.Min)
  n123 = MaterialFunctionCall(Color=Subtract(A=@n121.Max, B=@n121.Max)) {materialFunction=&SplitComponents}
  n124 = Max(A=Max(A=@n123.R, B=@n123.R), B=@n123.B)
  NamedRerouteDeclaration(None=FunctionInput(Preview=Divide(A=@n122, B=@n124)) {bUsePreviewValueAsDefault=true, inputName=Calcated_Bounds, sortPriority=11}) {name=Caclcated_Bounds, nodeColor=<rgba 0 0.9297 1 1>}
  n125 = TextureObjectParameter() {group=Wrap_Lancelot, parameterName=Lance__Tex_N, samplerType=SAMPLERTYPE_Normal, texture=&FlatNormalBlue}
  NamedRerouteDeclaration(None=@n125) {name=Text_N, nodeColor=<rgba 0 0.1562 1 1>}
  n126 = PreSkinnedNormal()
  NamedRerouteDeclaration(None=FunctionInput(Preview=VertexInterpolator(VS=@n126).PS) {bUsePreviewValueAsDefault=true, inputName=Pre-Skinned Local Normal, sortPriority=10}) {name=PreSkin_Normal, nodeColor=<rgba 0.8281 1 0 1>}
  NamedRerouteDeclaration(None=@n45.Normal) {name=Root_Normal, nodeColor=<rgba 0.4062 0 1 1>}
  NamedRerouteUsage()
  Power(Base=@n120, Exp=ScalarParameter() {defaultValue=0.25, group=TopSideProjection_Emissive_Reactive, parameterName=_Falloff})
  Power(Base=@n19)
  ScalarParameter() {defaultValue=0.1, group=Wrap_Lancelot, parameterName=Lance_Curption_Size}
  ScalarParameter() {defaultValue=-9999999, group=TopSideProjection_Emissive_Reactive, parameterName=Time_BP}
 ]}
