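# Material function for the Mitosis wrap skin, layering animated corruption veins and glow over weapons, vehicles, mechs, characters, backpacks and pickaxes.
# Projections from pre-skinned position use per-item-type offset, scale and rotation selected by IsState inputs.
# Noise-warped vein masks and time-cycled noise index curve-atlas colors, with normals derived from offset curve samples.
# Input masks blend base and accent layers; the Wrap_Mitosis static switch otherwise passes the original material through.
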MaterialFunction /Game/Weapons/WeaponSkins/Wraps/MaterialFunctions/MF_WrapMitos
graph {nodes [
  n1 = NamedRerouteUsage()
  n2 = Saturate(None=Multiply(A=Add(A=@n1) {constB=0.5}))
  n3 = Add(A=@n1) {constB=-0.5}
  n4 = Add(A=Multiply(A=OneMinus(None=@n2)) {constB=3}, B=Saturate(None=Multiply(A=@n3)))
  Add(B=@n4)
  n5 = TextureSample() {samplerType=SAMPLERTYPE_Masks, texture=/Game/Weapons/FORT_Rifles/Textures/T_Scar_CM}
  n6 = FunctionInput(Preview=@n5.R) {bUsePreviewValueAsDefault=true, inputName=Mask Primary, inputType=FunctionInput_Scalar, sortPriority=24, previewValue=<1 0 0 1>}
  n7 = FunctionInput(Preview=@n5.B) {bUsePreviewValueAsDefault=true, inputName=Mask Secondary, inputType=FunctionInput_Scalar, sortPriority=24, previewValue=<1 0 0 1>}
  Add(A=@n6, B=@n7)
  n8 = NamedRerouteUsage()
  n9 = Divide(A=ComponentMask(None=Add(A=@n8) {constB=-0.93}) {r=true}) {constB=0.05}
  n10 = Multiply(A=NamedRerouteUsage(), B=ScalarParameter() {defaultValue=0.5, group=Wrap_Mitosis_Corruption, parameterName=Mitosis_NoiseB_Amount_B}) {constB=0.5}
  n11 = Multiply(A=NamedRerouteUsage(), B=ScalarParameter() {defaultValue=0.75, group=Wrap_Mitosis_Corruption, parameterName=Mitosis_NoiseC_Amount_B}) {constB=0.5}
  n12 = Frac(None=Multiply(A=Add(A=@n10, B=@n11)) {constB=2})
  Add(A=Saturate(None=Power(Base=@n9)), B=@n12)
  n13 = NamedRerouteUsage()
  Add(A=@n13, B=NamedRerouteUsage())
  n14 = NamedRerouteUsage()
  n15 = NamedRerouteUsage()
  n16 = Time()
  n17 = MaterialFunctionCall(Float3=NamedRerouteUsage()) {materialFunction=&BreakOutFloat3Components}
  n18 = Multiply(A=@n16, B=@n17.R) {constB=10}
  n19 = MaterialFunctionCall(X=@n18, Y=Multiply(A=@n16, B=@n17.G) {constB=8}, Z=Multiply(A=@n16, B=@n17.B) {constB=-10}) {materialFunction=&MakeFloat3}
  n20 = PreSkinnedPosition()
  n21 = Constant3Vector() {constant=<rgba 1 1 1 0>}
  n22 = ShaderStageSwitch(PixelShader=VertexInterpolator(VS=Multiply(A=@n20, B=@n21)).PS, VertexShader=@n20)
  n23 = Add(A=@n19.Result, B=@n22)
  n24 = NamedRerouteUsage()
  ComponentMask(None=MaterialFunctionCall(
    Export Float 4=StaticBool() {value=true}
    TextureObject=@n14
    TextureSize=@n15
    WorldPosition=@n23
    World Space Normal=@n24
    ProjectionTransitionContrast=Constant()
   ) {materialFunction=&/WorldAlignedTexture}.XYZ Texture) {a=true}
  Constant3Vector()
  Constant4Vector()
  Constant()
  n25 = BlendMaterialAttributes(A=NamedRerouteUsage(), B=NamedRerouteUsage(), Alpha=Saturate(None=@n6)) {vertexAttributeBlendType=UseA}
  n26 = NamedRerouteUsage()
  n27 = NamedRerouteDeclaration(None=@n7) {name=Msk_Accent, nodeColor=<rgba 0 1 0.7656 1>}
  n28 = TextureSample() {samplerType=SAMPLERTYPE_Normal, texture=/Game/Weapons/FORT_Rifles/T_SCAR_N}
  n29 = Constant3Vector() {constant=<rgba 1 0 0 0>}
  n30 = GetMaterialAttributes(None=FunctionInput(Preview=SetMaterialAttributes!(MaterialAttributes, Normal=@n28.RGB, BaseColor=@n29) {attributeSetTypes [#1, #2]}) {inputName=Material, inputType=FunctionInput_MaterialAttributes}) {attributeGetTypes [
    #2
    #3
    #1
    #4
   ]}
  FunctionOutput(None=StaticSwitchParameter(
    True=BlendMaterialAttributes(A=@n25, B=@n26, Alpha=@n27) {vertexAttributeBlendType=UseA}
    False=NamedRerouteDeclaration(None=@n30.MaterialAttributes) {name=Root_Material, nodeColor=<rgba 0.7812 1 0 1>}
   ) {group=Customization_Wrap, parameterName=Wrap_Mitosis}) {outputName=Result_1}
  n31 = CurveAtlasRowParameter(CurveTime=@n12) {
   atlas=/Game/Weapons/WeaponSkins/Wraps/Textures/CruveAtlas_Wraps_01
   curve=/Game/Weapons/WeaponSkins/Wraps/Textures/0_ColorCurves/CC_Mitos_Coruption_Color
   defaultValue=84
   group=Wrap_Mitosis_Corruption
   parameterName=Mitosis_Corruption_Colors_inner_CC
   primitiveDataIndex=1
   sortPriority=0
  }
  n32 = Add(A=NamedRerouteUsage()) {constB=-0.5}
  LinearInterpolate(
   A=@n31
   B=Multiply(A=@n31, B=ScalarParameter() {defaultValue=2, group=Wrap_Mitosis_Corruption, parameterName=Mitosis_Corruption_Edge_Brightness})
   Alpha=Saturate(None=Multiply(A=@n32) {constB=2})
  )
  n33 = NamedRerouteUsage()
  n34 = NamedRerouteUsage()
  n35 = NamedRerouteUsage()
  n36 = NamedRerouteUsage()
  n37 = NamedRerouteUsage()
  n38 = VectorParameter() {group=Wrap_Mitosis_Corruption, parameterName=Mitosis_Veins_Noise_Speed, defaultValue=<rgba 0 0 0 1>}
  n39 = Add(
   A=Multiply(A=NamedRerouteUsage(), B=ScalarParameter() {defaultValue=1, group=Wrap_Mitosis_Corruption, parameterName=Mitosis_Veins_Noise_TileRate})
   B=Multiply(A=Time(), B=ComponentMask(None=@n38.RGB) {g=true, r=true})
  )
  n40 = NamedRerouteUsage()
  n41 = Multiply(
   A=Add(A=TextureSample(UVs=@n39, Tex=@n40).A) {constB=-0.5}
   B=ScalarParameter() {defaultValue=0.1, group=Wrap_Mitosis_Corruption, parameterName=Mitosis_Veins_Noise_Warp Amount}
  )
  MaterialFunctionCall!(
   Normal_Map=@n33
   Offset Rt
   Pivot
   Caculated_Bounds=@n34
   Pre_Skinned_Normals=@n35
   is_State=NamedRerouteUsage()
   UV_Rotation=NamedRerouteUsage()
   Uvs_Offset=@n41
   Uvs_TileRate=@n36
   Uvs_Offset_Pivot=@n37
   For_Safe_Zones
   Rotation_Comp
   Clamp Y
   Clamp X
   Clamp Padding_Comp
   No_Rotation
  ) {materialFunction=&/MF_Projection_Side_Normal_V2}
  n42 = ScalarParameter() {defaultValue=84, group=Wrap_Mitosis_Corruption, parameterName="Mitosis_Corruption_Colors_inner_CC  Index", sortPriority=0}
  n43 = Multiply(A=Abs(None=@n15)) {constB=-1}
  n44 = MaterialFunctionCall(Float4=NamedRerouteUsage()) {materialFunction=&BreakOutFloat4Components}
  n45 = TextureSample(
   UVs=MaterialFunctionCall!(
    For_Safe_Zones
    Normal_Map
    Caculated_Bounds=Divide(A=@n23, B=@n43)
    Pre_Skinned_Normals=@n24
    Is Vehicle=@n44.R
    Is VehicleSide=@n44.R
    Is Mech=@n44.R
    Vehicle_Rotation
    UV_Rotation Bypass
    Mech_Rotation
    Uvs_Offset
    Offset-after rotation
    Uvs_TileRate
   ) {materialFunction=&/MF_Projection_Side_Normal}.UVs No Offsets/Tiling
   Tex=@n14
  ) {samplerType=SAMPLERTYPE_LinearColor}
  n46 = ShaderStageSwitch(
   PixelShader=Divide(A=@n45.A, B=ScalarParameter() {defaultValue=5, group=Wrap_Mitosis_Corruption, parameterName=Mitosis_Noise_Added_Strength}) {constB=5}
   VertexShader=Constant()
  )
  n47 = NamedRerouteUsage()
  n48 = Add(A=Divide(A=@n19.Result, B=@n47) {constB=300}, B=Divide(A=@n22, B=@n47) {constB=300})
  n49 = Time()
  n50 = DotProduct(
   A=TextureSample(UVs=Add(A=@n46, B=@n48), Tex=NamedRerouteUsage()) {samplerType=SAMPLERTYPE_LinearColor}.RGB
   B=MaterialFunctionCall(Hue Shift Percentage=Divide(A=@n49) {constB=5}, Texture=Constant3Vector() {constant=<rgba 1 0 0 0>}) {materialFunction=&HueShift}.Result
  )
  n51 = Divide(A=Time(), B=ScalarParameter() {defaultValue=4, group=Wrap_Mitosis_Corruption, parameterName=Mitosis_FracSpeed})
  n52 = Add(
   A=Multiply(A=@n50, B=ScalarParameter() {defaultValue=3, group=Wrap_Mitosis_Corruption, parameterName=Mitosis_Noise_FracMultiplier}) {constB=4}
   B=@n51
  )
  n53 = ScalarParameter() {
   defaultValue=0.19
   desc="The value of the black color in the gradient, this combined with the mask prevents black color to show in veins"
   group=Wrap_Mitosis_Corruption
   parameterName=Mitosis_BlackPoint_Value
  }
  n54 = ComponentMask(None=Add(A=NamedRerouteUsage(), B=ScalarParameter() {defaultValue=-0.95, group=Wrap_Mitosis_Corruption, parameterName=Mitosis_BlackPoint_Offset}) {constB=-0.95}) {r=true}
  n55 = ScalarParameter() {defaultValue=-0.3, group=Wrap_Mitosis_Corruption, parameterName=Mitosis_BlackPoint_Falloff}
  n56 = Saturate(None=Power(Base=Divide(A=@n54, B=@n55) {constB=-0.3}))
  n57 = MaterialFunctionCall(Input=Frac(None=@n52), Target Low=Saturate(None=Multiply(A=@n53, B=@n56))) {materialFunction=&RemapValueRange}
  n58 = ScalarParameter() {defaultValue=0.01, group=Wrap_Mitosis_Corruption, parameterName=Mitosis_Noise_Normal}
  n59 = Add(A=@n57.Result, B=@n58) {constB=0.01}
  n60 = CurveAtlasRowParameter(CurveTime=@n59) {
   atlas=&CruveAtlas_Wraps_01
   curve=&CC_Mitos_Coruption_Color
   defaultValue=84
   group=Wrap_Mitosis_Corruption
   parameterName=Mitosis_Corruption_Colors_inner_CC
   primitiveDataIndex=1
   sortPriority=0
  }
  n61 = MaterialFunctionCall!(Index=@n42, CCinput, CCinput_Alpha=@n60.A, UV=@n59) {materialFunction=&/MF_Global_CC_Tex}
  n62 = CurveAtlasRowParameter(CurveTime=@n57.Result) {
   atlas=&CruveAtlas_Wraps_01
   curve=&CC_Mitos_Coruption_Color
   defaultValue=84
   group=Wrap_Mitosis_Corruption
   parameterName=Mitosis_Corruption_Colors_inner_CC
   primitiveDataIndex=1
   sortPriority=0
  }
  n63 = MaterialFunctionCall(Index=@n42, CCinput=@n62, CCinput_Alpha=@n62, UV=@n57.Result) {materialFunction=&MF_Global_CC_Tex}
  n64 = Subtract(A=@n57.Result, B=@n58)
  n65 = CurveAtlasRowParameter(CurveTime=@n64) {
   atlas=&CruveAtlas_Wraps_01
   curve=&CC_Mitos_Coruption_Color
   defaultValue=84
   group=Wrap_Mitosis_Corruption
   parameterName=Mitosis_Corruption_Colors_inner_CC
   primitiveDataIndex=1
   sortPriority=0
  }
  n66 = MaterialFunctionCall(Index=@n42, CCinput_Alpha=@n65.A, UV=@n64) {materialFunction=&MF_Global_CC_Tex}
  n67 = CrossProduct(
   A=AppendVector(A=Constant2Vector() {r=1}, B=Subtract(A=@n61.Alpha, B=@n63.Alpha))
   B=AppendVector(A=Constant2Vector() {g=1}, B=Subtract(A=@n66.Alpha, B=@n63.Alpha))
  )
  MaterialFunctionCall(AdditionalNormal=Normalize(VectorInput=@n67)) {materialFunction=&BlendAngleCorrectedNormals}
  Multiply(A=@n45.A)
  n68 = MaterialFunctionCall(Float4=NamedRerouteUsage()) {materialFunction=&BreakOutFloat4Components}
  n69 = MaterialFunctionCall!(
   Normal_Map=@n33
   Caculated_Bounds=@n34
   Pre_Skinned_Normals=@n35
   Skew Amount=Constant() {r=0.5}
   Is Vehicle=@n68.R
   Uvs_Offset=Add(A=@n37, B=@n41)
   Uvs_TileRate=Multiply(A=@n36, B=Constant3Vector() {constant=<rgba 1 1 1 0>})
   Alternate Axies contrast=Constant() {r=1}
  ) {materialFunction=&/MF_Projection_Skew_Normal}
  n70 = NamedRerouteUsage()
  n71 = TextureSample(UVs=@n69.UVs, Tex=@n70)
  n72 = NamedRerouteUsage()
  n73 = NamedRerouteUsage()
  n74 = NamedRerouteUsage()
  n75 = Add(A=@n41, B=Multiply(A=Add(A=@n73, B=@n74) {constB=0}, B=NamedRerouteUsage()))
  n76 = NamedRerouteUsage()
  n77 = Constant()
  n78 = ScalarParameter() {defaultValue=5, group=Wrap_Mitosis, parameterName=Mitosis_Contrast, sortPriority=10}
  n79 = MaterialFunctionCall(Float2=NamedRerouteUsage()) {materialFunction=&BreakOutFloat2Components}
  n80 = MaterialFunctionCall(Float4=NamedRerouteUsage()) {materialFunction=&BreakOutFloat4Components}
  n81 = Max(A=Max(A=@n79.R, B=@n79.R), B=Max(A=Max(A=@n80.R, B=@n80.R), B=Max(A=@n80.B, B=@n80.B)))
  n82 = LinearInterpolate(
   A=@n71.RGBA
   B=MaterialFunctionCall!(
    IsCharacter=@n68.A
    IsStatsB=NamedRerouteUsage()
    Texture=@n70
    Rotation_Axis=@n72
    Caculated_Bounds=@n75
    Pre_Skinned_Normals=@n76
    Blend_Contrast=@n78
    Fill_alternate_Axis_Opacity=@n77
    "Fill_alternate_Axis_Color_A=Roughness"=Constant4Vector().RGBA
    Is_Vehicle=@n68.A
    Is_Mech=@n68.A
    Enable_Fill_Z_Weapon
    Fill_Z_Weapon_opacity
    Fill_Z_Weapon_Color
   ) {materialFunction=&/MF_TopsSideProjection_Base}.Combined
   Alpha=@n81
  )
  n83 = ChannelMaskParameter(None=@n82) {group=Wrap_Mitosis_Corruption, parameterName=Mitosis_Corruption_Channel}
  Multiply(A=@n83.R)
  n84 = MaterialFunctionCall(Float4=NamedRerouteUsage()) {materialFunction=&BreakOutFloat4Components}
  n85 = MaterialFunctionCall(
   Caculated_Bounds=NamedRerouteUsage()
   Pre_Skinned_Normals=NamedRerouteUsage()
   Is Vehicle=@n84.R
   Is VehicleSide=@n84.R
   Is Mech=@n84.R
   UV_Rotation Bypass=NamedRerouteUsage()
   Uvs_Offset=NamedRerouteUsage()
  ) {materialFunction=&MF_Projection_Side_Normal}
  Multiply(
   A=TextureSample(UVs=NamedRerouteDeclaration(None=@n85.UVs) {name=Uvs, nodeColor=<rgba 1 0.4219 0 1>}, Tex=NamedRerouteUsage()).R
   B=NamedRerouteUsage()
  )
  Multiply(A=MaterialFunctionCall(Value_Morning=Constant() {r=0.5}, Value_Day=Constant() {r=1}, Value_Evening=Constant() {r=0.5}, Value_Night=Constant() {r=0.25}) {materialFunction=&Scalar_Time_of_Day}.Result)
  n86 = Add(A=Multiply(A=NamedRerouteUsage()), B=Multiply(A=Time(), B=Constant() {r=1}))
  Multiply(A=Frac(None=ComponentMask(None=@n86) {r=true}))
  Multiply(A=@n71.R)
  n87 = Normalize(VectorInput=@n22)
  n88 = VectorParameter() {group=Wrap_Mitosis_Corruption, parameterName=Mitosis_Noise_SidesMaskDirection, defaultValue=<rgba 0.5 -0.3 0 1>}
  n89 = Normalize(VectorInput=@n88.RGB)
  Multiply(A=@n18, B=Sign(None=DotProduct(A=@n87, B=@n89)))
  Multiply()
  Multiply(A=@n52)
  n90 = VectorParameter() {group=Wrap_Mitosis_Accent, parameterName=Mitosis_Accent_BaseColor, defaultValue=<rgba 0.005 0.005 0.005 1>}
  n91 = SetMaterialAttributes!(
   MaterialAttributes=NamedRerouteUsage()
   BaseColor=@n90.RGB
   Roughness=ScalarParameter() {defaultValue=0.4, group=Wrap_Mitosis_Accent, parameterName=Mitosis_Accent_Roughness}
   Specular=ScalarParameter() {defaultValue=0.1, group=Wrap_Mitosis_Accent, parameterName=Mitosis_Accent_Spec}
   Metallic=ScalarParameter() {group=Wrap_Mitosis_Accent, parameterName=Mitosis_Accent_Metalic}
   Normal
   EmissiveColor=Constant()
  ) {attributeSetTypes [
    #2
    #3
    #5
    #4
    #1
    #6
   ]}
  NamedRerouteDeclaration(None=@n91) {name=Mat_Accent, nodeColor=<rgba 0 0.3906 1 1>}
  n92 = FunctionInput(Preview=@n5.G) {bUsePreviewValueAsDefault=true, inputName=Mask Crevise, inputType=FunctionInput_Scalar, sortPriority=24, previewValue=<1 0 0 1>}
  NamedRerouteDeclaration(None=@n92) {name=Msk_Curvature, nodeColor=<rgba 0 1 0.7422 1>}
  n93 = Multiply(
   A=ScalarParameter() {defaultValue=1.5, group=Wrap_Mitosis_Fresnel, parameterName=Mitosis_ThinFilm_Scale}
   B=Fresnel(ExponentIn=ScalarParameter() {defaultValue=2, group=Wrap_Mitosis_Fresnel, parameterName=Mitosis_ThinFilm_Warp}) {baseReflectFraction=0}
  )
  n94 = CurveAtlasRowParameter(CurveTime=@n93) {
   atlas=&CruveAtlas_Wraps_01
   curve=/Game/Weapons/WeaponSkins/Wraps/Textures/0_ColorCurves/CC_Mitos_ThinFilm
   defaultValue=82
   group=Wrap_Mitosis_Fresnel
   parameterName=Mitosis_ThinFilm_CC
   primitiveDataIndex=1
   sortPriority=-5
  }
  NamedRerouteDeclaration(None=MaterialFunctionCall(
    Index=ScalarParameter() {defaultValue=82, group=Wrap_Mitosis_Fresnel, parameterName="Mitosis_ThinFilm_CC  Index", sortPriority=-5}
    CCinput=@n94
    UV=@n93
   ) {materialFunction=&MF_Global_CC_Tex}.3V) {name=Thin_Film, nodeColor=<rgba 1 0.5859 0 1>}
  n95 = TextureObjectParameter() {
   group=Wrap_Mitosis
   parameterName=Mitosis__Tex_Msk1
   samplerType=SAMPLERTYPE_LinearColor
   texture=/Game/Weapons/WeaponSkins/Wraps/Textures/Mitosis/T_Wrap_Mitosis_MSk
  }
  NamedRerouteDeclaration(None=@n95) {name=Tex_Msk1, nodeColor=<rgba 0.1953 1 0 1>}
  n96 = TextureObjectParameter() {
   group=Wrap_Mitosis
   parameterName=Mitosis__Tex_Msk2
   samplerType=SAMPLERTYPE_LinearColor
   texture=/Game/Weapons/WeaponSkins/Wraps/Textures/Mitosis/T_Wrap_Mitosis_Msk1
  }
  NamedRerouteDeclaration(None=@n96) {name=Tex_Msk2, nodeColor=<rgba 0.1953 1 0 1>}
  n97 = ScalarParameter() {defaultValue=83, group=Wrap_Mitosis_Corruption, parameterName="Mitosis_Corruption_Hight_CC  Index", sortPriority=-5}
  n98 = Time()
  n99 = VectorParameter() {group=Wrap_Mitosis_Corruption, parameterName=Mitosis_NoiseA_Speed, defaultValue=<rgba 0.1 0.003 0 1>}
  n100 = Add(
   A=Multiply(A=NamedRerouteUsage(), B=ScalarParameter() {defaultValue=8, group=Wrap_Mitosis_Corruption, parameterName=Mitosis_NoiseA_TileRate})
   B=ComponentMask(None=Multiply(A=@n98, B=@n99.RGB)) {g=true, r=true}
  )
  n101 = NamedRerouteUsage()
  n102 = Add(
   A=@n83.R
   B=Multiply(
    A=TextureSample(UVs=@n100, Tex=@n101) {samplerType=SAMPLERTYPE_LinearGrayscale, texture=&T_Wrap_Mitosis_MSk}.R
    B=ScalarParameter() {defaultValue=0.16, group=Wrap_Mitosis_Corruption, parameterName=Mitosis_NoiseA_Amount_A}
   ) {constB=0.1}
  )
  n103 = CurveAtlasRowParameter(CurveTime=@n102) {
   atlas=&CruveAtlas_Wraps_01
   curve=/Game/Weapons/WeaponSkins/Wraps/Textures/0_ColorCurves/CC_Mitos_Coruption
   defaultValue=83
   group=Wrap_Mitosis_Corruption
   parameterName=Mitosis_Corruption_Hight_CC
   primitiveDataIndex=1
   sortPriority=-5
  }
  MaterialExpressionMaterialFunctionCall_32 = MaterialFunctionCall(Index=@n97, CCinput=@n103, CCinput_Alpha=@n103, UV=@n102) {materialFunction=&MF_Global_CC_Tex}
  n104 = Convert(None=@MaterialExpressionMaterialFunctionCall_32.4v) {
   nodeName=Break Float4
   convertInputs [{
     type=Vector4
     defaultValue=<rgba 0 0 0 1>
     expressionInput {
      expression=&self:MaterialExpressionMaterialFunctionCall_32
      inputName=None
      mask=0
      maskA=0
      maskB=0
      maskG=0
      maskR=0
      outputIndex=1
     }
    }]
   convertOutputs [{type=Scalar}, {type=Scalar}, {type=Scalar}, {type=Scalar}]
  }
  n105 = NamedRerouteUsage()
  NamedRerouteDeclaration(None=Multiply(A=SmoothStep(Value=@n104) {constMax=0.2, constMin=0.15}, B=Multiply(A=@n104, B=@n105)) {constA=1}) {name=Corruption Colors, nodeColor=<rgba 0 1 0.2031 1>}
  NamedRerouteDeclaration(None=@n83.R) {name=Corruption Colors_Msk, nodeColor=<rgba 0 1 0.2031 1>}
  n106 = CurveAtlasRowParameter(CurveTime=@n83.R) {
   atlas=&CruveAtlas_Wraps_01
   curve=&CC_Mitos_Coruption
   defaultValue=83
   group=Wrap_Mitosis_Corruption
   parameterName=Mitosis_Corruption_Hight_CC
   primitiveDataIndex=1
   sortPriority=-5
  }
  n107 = Power(Base=OneMinus(None=MaterialFunctionCall(Index=@n97, CCinput_Alpha=@n106.B, UV=@n83.R) {materialFunction=&MF_Global_CC_Tex}.Alpha))
  NamedRerouteDeclaration(None=Saturate(None=OneMinus(None=@n107))) {name=Corruption Shadow_Msk, nodeColor=<rgba 0 1 0.2031 1>}
  n108 = NamedRerouteUsage()
  n109 = NamedRerouteUsage()
  n110 = TextureSample(
   UVs=Multiply(A=@n109, B=ScalarParameter() {defaultValue=1, group=Wrap_Mitosis, parameterName=Mitosis_Roughness_TileRate, sortPriority=20})
   Tex=NamedRerouteUsage()
  )
  n111 = Multiply(A=Multiply(A=@n110.R) {constB=2}, B=NamedRerouteUsage())
  n112 = NamedRerouteUsage()
  n113 = Saturate(None=Multiply(A=Add(A=@n112) {constB=0.5}))
  n114 = Add(A=@n112) {constB=-0.5}
  n115 = OneMinus(None=Saturate(None=Multiply(A=@n114) {constB=3}))
  n116 = MaterialFunctionCall(
   A=VectorParameter() {group=Wrap_Mitosis, parameterName=Mitosis_Color_A, defaultValue=<rgba 0.0043 0.0048 0.005 1>}.RGB
   B=VectorParameter() {group=Wrap_Mitosis, parameterName=Mitosis_Color_B, defaultValue=<rgba 0.0247 0.0316 0.04 1>}.RGB
   C=VectorParameter() {group=Wrap_Mitosis, parameterName=Mitosis_Color_C, defaultValue=<rgba 0.0545 0.0802 0.0976 1>}.RGB
   Alpha=Multiply(A=Multiply(A=@n111, B=@n113), B=@n115)
  ) {materialFunction=&Lerp_3Color}
  n117 = Saturate(None=Add(A=@n115) {constB=0.25})
  n118 = Multiply(A=NamedRerouteUsage(), B=@n117)
  n119 = NamedRerouteUsage()
  n120 = NamedRerouteUsage()
  n121 = ScalarParameter() {defaultValue=1, group=Wrap_Mitosis, parameterName=Mitosis_Curruption_BC_Brightness}
  n122 = NamedRerouteUsage()
  n123 = LinearInterpolate(
   A=Multiply(A=Add(A=@n116.Result, B=@n118), B=Multiply(A=@n119, B=@n117))
   B=LinearInterpolate(
    A=Multiply(A=@n120, B=@n121)
    B=VectorParameter() {group=Wrap_Mitosis, parameterName=Mitosis_Glow_Front, defaultValue=<rgba 1 1 1 1>}.RGB
    Alpha=NamedRerouteUsage()
   )
   Alpha=@n122
  )
  n124 = VectorParameter() {group=Wrap_Mitosis, parameterName=Mitosis_Glow_Color, channelNames {a=Darkens BC, b="", g="", r=""}, defaultValue=<rgba 1 1 1 1>}
  n125 = NamedRerouteUsage()
  n126 = ScalarParameter() {defaultValue=0.3, group=Wrap_Mitosis, parameterName=Mitosis_Roughness_Low, sortPriority=20}
  n127 = ScalarParameter() {defaultValue=1, group=Wrap_Mitosis, parameterName=Mitosis_Roughness_High, sortPriority=20}
  n128 = Saturate(None=@n4)
  n129 = LinearInterpolate(A=@n110.R, Alpha=@n128)
  n130 = NamedRerouteUsage()
  n131 = ScalarParameter() {defaultValue=0.13, group=Wrap_Mitosis, parameterName=Mitosis_Spec_High, sortPriority=20}
  n132 = ScalarParameter() {group=Wrap_Mitosis, parameterName=Mitosis_Spec_Low, sortPriority=20}
  n133 = ScalarParameter() {defaultValue=10, group=Wrap_Mitosis, parameterName=Mitosis_Normal_TileRate}
  n134 = Transform(None=TextureSample(UVs=Multiply(A=@n109, B=@n133), Tex=NamedRerouteUsage()) {samplerType=SAMPLERTYPE_LinearColor}.RGB)
  n135 = Multiply(
   A=Multiply(
    A=Transform(None=@n134) {transformSourceType=TRANSFORMSOURCE_World, transformType=TRANSFORM_Tangent}
    B=Constant3Vector() {constant=<rgba 1 1 0 0>}
   )
   B=ScalarParameter() {defaultValue=0.5, group=Wrap_Mitosis, parameterName=Mitosis_Normal_Strength}
  )
  n136 = SetMaterialAttributes(
   MaterialAttributes=@n108
   BaseColor=LinearInterpolate(A=@n123, B=Multiply(A=@n123, B=@n124.Darkens BC), Alpha=@n125)
   Roughness=LinearInterpolate(
    A=LinearInterpolate(A=@n126, B=@n127, Alpha=@n129)
    B=ScalarParameter() {defaultValue=0.25, group=Wrap_Mitosis, parameterName=Mitosis_Roughness_Corruption, sortPriority=20}
    Alpha=@n130
   )
   Specular=LinearInterpolate(
    A=LinearInterpolate(A=@n131, B=@n132, Alpha=@n128)
    B=ScalarParameter() {defaultValue=0.25, group=Wrap_Mitosis, parameterName=Mitosis_Spec_Corruption, sortPriority=20}
    Alpha=@n130
   )
   Metallic=ScalarParameter() {group=Wrap_Mitosis, parameterName=Mitosis_Metalic, sortPriority=20}
   Normal=Add(A=NamedRerouteUsage(), B=LinearInterpolate(A=@n135, B=@n13, Alpha=@n122))
   EmissiveColor=LinearInterpolate(B=Multiply(A=@n123, B=@n124), Alpha=@n125)
  ) {attributeSetTypes [
    #2
    #3
    #5
    #4
    #1
    #6
   ]}
  NamedRerouteDeclaration(None=@n136) {name=Mat_Base, nodeColor=<rgba 0 0.3906 1 1>}
  NamedRerouteDeclaration(None=@n104) {name=Corruption Colors_Height, nodeColor=<rgba 0 1 0.2031 1>}
  n137 = FunctionInput(Preview=@n5.A) {bUsePreviewValueAsDefault=true, inputName=Mask Gmap, inputType=FunctionInput_Scalar, sortPriority=24, previewValue=<1 0 0 1>}
  NamedRerouteDeclaration(None=@n137) {name=Msk_Gmap, nodeColor=<rgba 0 1 0.7656 1>}
  n138 = Transform(None=MaterialFunctionCall!(VectorToRotate=VertexNormalWS(), LookAtVector=@n67, RestingVector) {materialFunction=/Engine/Functions/Engine_MaterialFunctions01/Vectors/RotateVector}.Result) {transformSourceType=TRANSFORMSOURCE_World, transformType=TRANSFORM_Tangent}
  NamedRerouteDeclaration(None=Multiply(A=Normalize(VectorInput=@n138), B=Constant3Vector() {constant=<rgba 1 1 0 0>})) {name=Generated_normals, nodeColor=<rgba 0 1 0.5547 1>}
  NamedRerouteDeclaration(None=@n63.3V) {name=Corruption _Colors_interror, nodeColor=<rgba 0 1 0.4141 1>}
  n139 = OneMinus(None=@n104)
  n140 = CurveAtlasRowParameter(CurveTime=@n139) {
   atlas=&CruveAtlas_Wraps_01
   curve=&CC_Mitos_Coruption
   defaultValue=83
   group=Wrap_Mitosis_Corruption
   parameterName=Mitosis_Corruption_Hight_CC
   primitiveDataIndex=1
   sortPriority=-5
  }
  n141 = MaterialFunctionCall(Index=@n97, CCinput_Alpha=@n140.A, UV=@n139) {materialFunction=&MF_Global_CC_Tex}
  NamedRerouteDeclaration(None=OneMinus(None=Power(Base=@n141.Alpha) {constExponent=5})) {name=Corruption Colors_Height_B, nodeColor=<rgba 0 1 0.2031 1>}
  n142 = ScalarParameter() {group=Wrap_Mitosis_UVs, parameterName=Mitosis_Rotation_Vehicle, sortPriority=150}
  n143 = ScalarParameter() {group=Wrap_Mitosis_UVs, parameterName=Mitosis_Rotation_Vehicle_Side, sortPriority=150}
  n144 = MaterialFunctionCall(Float4=NamedRerouteUsage()) {materialFunction=&BreakOutFloat4Components}
  n145 = LinearInterpolate(
   A=LinearInterpolate(
    A=ScalarParameter() {group=Wrap_Mitosis_UVs, parameterName=Mitosis_Rotation_Weapon, sortPriority=150}
    B=LinearInterpolate(A=@n142, B=@n143, Alpha=@n144.G)
    Alpha=@n144.R
   )
   B=LinearInterpolate(
    A=ScalarParameter() {group=Wrap_Mitosis_UVs, parameterName=Mitosis_Rotation_Mech, sortPriority=150}
    B=ScalarParameter() {group=Wrap_Mitosis_UVs, parameterName=Mitosis_Rotation_Character, sortPriority=150}
    Alpha=@n144.A
   )
   Alpha=@n144.B
  )
  n146 = ScalarParameter() {group=Wrap_Mitosis_UVs, parameterName=Mitosis_Rotation_Backpack, sortPriority=150}
  n147 = MaterialFunctionCall(Float2=NamedRerouteUsage()) {materialFunction=&BreakOutFloat2Components}
  NamedRerouteDeclaration(None=LinearInterpolate(
    A=LinearInterpolate(A=@n145, B=@n146, Alpha=@n147.R)
    B=ScalarParameter() {group=Wrap_Mitosis_UVs, parameterName=Mitosis_Rotation_Pickaxe, sortPriority=150}
    Alpha=@n147.G
   )) {name=UVs_Rotation, nodeColor=<rgba 0 0.7188 1 1>}
  n148 = VectorParameter() {group=Wrap_Mitosis_UVs, parameterName=Mitosis_Scale_Weapon, defaultValue=<rgba 1 1 1 1>}
  n149 = VectorParameter() {group=Wrap_Mitosis_UVs, parameterName=Mitosis_Scale_Vehicle, defaultValue=<rgba 1 1 1 1>}
  n150 = Multiply(A=@n149.RGB, B=@n149.RGB)
  n151 = VectorParameter() {group=Wrap_Mitosis_UVs, parameterName=Mitosis_Scale_VehicleSide, defaultValue=<rgba 1 1 1 1>}
  n152 = Multiply(A=@n151.RGB, B=@n151.RGB)
  n153 = MaterialFunctionCall(Float4=NamedRerouteUsage()) {materialFunction=&BreakOutFloat4Components}
  n154 = VectorParameter() {group=Wrap_Mitosis_UVs, parameterName=Mitosis_Scale_Mech, defaultValue=<rgba 1 1 1 1>}
  n155 = VectorParameter() {group=Wrap_Mitosis_UVs, parameterName=Mitosis_Scale_Character, defaultValue=<rgba 1 1 1 1>}
  n156 = LinearInterpolate(
   A=LinearInterpolate(A=Multiply(A=@n148.RGB, B=@n148.RGB), B=LinearInterpolate(A=@n150, B=@n152, Alpha=@n153.G), Alpha=@n153.R)
   B=LinearInterpolate(A=Multiply(A=@n154.RGB, B=@n154.RGB), B=Multiply(A=@n155.RGB, B=@n155.RGB), Alpha=@n153.A)
   Alpha=@n153.B
  )
  n157 = VectorParameter() {group=Wrap_Mitosis_UVs, parameterName=Mitosis_Scale_Backpack, defaultValue=<rgba 1 1 1 1>}
  n158 = Multiply(A=@n157.RGB, B=@n157.RGB)
  n159 = MaterialFunctionCall(Float2=NamedRerouteUsage()) {materialFunction=&BreakOutFloat2Components}
  n160 = VectorParameter() {group=Wrap_Mitosis_UVs, parameterName=Mitosis_Scale_Pickaxe, defaultValue=<rgba 1 1 1 1>}
  NamedRerouteDeclaration(None=LinearInterpolate(A=LinearInterpolate(A=@n156, B=@n158, Alpha=@n159.R), B=Multiply(A=@n160.RGB, B=@n160.RGB), Alpha=@n159.G)) {name=UVs_Scale, nodeColor=<rgba 0 0.7188 1 1>}
  n161 = MaterialFunctionCall(Float3=Abs(None=NamedRerouteUsage())) {materialFunction=&BreakOutFloat3Components}
  n162 = Constant() {r=0.5}
  n163 = NamedRerouteUsage()
  n164 = ComponentMask(None=@n163) {b=true, desc=YZ, g=true}
  n165 = ComponentMask(None=@n163) {b=true, desc=XZ, r=true}
  NamedRerouteDeclaration(None=If(
    A=@n161.B
    B=@n162
    "A > B"=ComponentMask(None=@n163) {desc=YX, g=true, r=true}
    "A < B"=If(A=@n161.R, B=@n162, "A > B"=@n164, "A < B"=@n165)
   )) {name=UVS_Cheap_Tri, nodeColor=<rgba 1 0 0.8438 1>}
  n166 = VectorParameter() {group=Wrap_Mitosis_UVs, parameterName=Mitosis_Offset_Vehicle, defaultValue=<rgba 0 0 0 1>}
  n167 = VectorParameter() {group=Wrap_Mitosis_UVs, parameterName=Mitosis_Offset_VehicleSide, defaultValue=<rgba 0 0 -1 1>}
  n168 = MaterialFunctionCall(Float4=NamedRerouteUsage()) {materialFunction=&BreakOutFloat4Components}
  n169 = LinearInterpolate(
   A=LinearInterpolate(
    A=VectorParameter() {group=Wrap_Mitosis_UVs, parameterName=Mitosis_Offset_Weapon, defaultValue=<rgba 0 0 -1 1>}.RGB
    B=LinearInterpolate(A=@n166.RGB, B=@n167.RGB, Alpha=@n168.G)
    Alpha=@n168.R
   )
   B=LinearInterpolate(
    A=VectorParameter() {group=Wrap_Mitosis_UVs, parameterName=Mitosis_Offset_Mech, defaultValue=<rgba 0 0 0 1>}.RGB
    B=VectorParameter() {group=Wrap_Mitosis_UVs, parameterName=Mitosis_Offset_Character, defaultValue=<rgba 0 0 0 1>}.RGB
    Alpha=@n168.A
   )
   Alpha=@n168.B
  )
  n170 = VectorParameter() {group=Wrap_Mitosis_UVs, parameterName=Mitosis_Offset_Backpack, defaultValue=<rgba 0 0 0 1>}
  n171 = MaterialFunctionCall(Float2=NamedRerouteUsage()) {materialFunction=&BreakOutFloat2Components}
  NamedRerouteDeclaration(None=LinearInterpolate(
    A=LinearInterpolate(A=@n169, B=@n170.RGB, Alpha=@n171.R)
    B=VectorParameter() {group=Wrap_Mitosis_UVs, parameterName=Mitosis_Offset_Pickaxe, defaultValue=<rgba 0 0 0 1>}.RGB
    Alpha=@n171.G
   )) {name=UVs_Offset, nodeColor=<rgba 0 0.625 1 1>}
  n172 = TextureObjectParameter() {
   group=Wrap_Mitosis
   parameterName=Mitosis__Tex_N
   samplerType=SAMPLERTYPE_Normal
   texture=&FlatNormalBlue
  }
  NamedRerouteDeclaration(None=@n172) {name=Tex_Nrml, nodeColor=<rgba 0.1953 1 0 1>}
  n173 = VectorParameter() {group=Wrap_Mitosis_UVs, parameterName=Mitosis_Rotation_Vehicle, defaultValue=<rgba 0 0 0 1>}
  n174 = VectorParameter() {group=Wrap_Mitosis_UVs, parameterName=Mitosis_Rotation_Vehicle_Side, defaultValue=<rgba 0 0 0 1>}
  n175 = MaterialFunctionCall(Float4=NamedRerouteUsage()) {materialFunction=&BreakOutFloat4Components}
  n176 = LinearInterpolate(
   A=LinearInterpolate(
    A=VectorParameter() {group=Wrap_Mitosis_UVs, parameterName=Mitosis_Rotation_Weapon, defaultValue=<rgba 0 0 -0.25 1>}.RGB
    B=LinearInterpolate(A=@n173.RGB, B=@n174.RGB, Alpha=@n175.G)
    Alpha=@n175.R
   )
   B=LinearInterpolate(
    A=VectorParameter() {group=Wrap_Mitosis_UVs, parameterName=Mitosis_Rotation_Mech, defaultValue=<rgba 0.25 0.25 0 1>}.RGB
    B=VectorParameter() {group=Wrap_Mitosis_UVs, parameterName=Mitosis_Rotation_Character, defaultValue=<rgba 0.25 0.25 0 1>}.RGB
    Alpha=@n175.A
   )
   Alpha=@n175.B
  )
  n177 = VectorParameter() {group=Wrap_Mitosis_UVs, parameterName=Mitosis_Rotation_Backpack, defaultValue=<rgba 0 0 0 1>}
  n178 = MaterialFunctionCall(Float2=NamedRerouteUsage()) {materialFunction=&BreakOutFloat2Components}
  NamedRerouteDeclaration(None=LinearInterpolate(
    A=LinearInterpolate(A=@n176, B=@n177.RGB, Alpha=@n178.R)
    B=VectorParameter() {group=Wrap_Mitosis_UVs, parameterName=Mitosis_Rotation_Pickaxe, defaultValue=<rgba 0 0 0 1>}.RGB
    Alpha=@n178.G
   )) {name=TopSide_Rotation, nodeColor=<rgba 0 0.2969 1 1>}
  n179 = ChannelMaskParameter(None=@n82) {group=Wrap_Mitosis_Corruption, maskChannel=Alpha, parameterName=Mitosis_Glow_Mask_Channel, defaultValue=<rgba 0 0 0 1>}
  NamedRerouteDeclaration(None=@n179.R) {name=Corruption Glow_Msk, nodeColor=<rgba 0 1 0.2031 1>}
  n180 = MaterialFunctionCall(Color=Abs(None=NamedRerouteUsage())) {materialFunction=&SplitComponents}
  n181 = MaterialFunctionCall(Float4=NamedRerouteUsage()) {materialFunction=&BreakOutFloat4Components}
  n182 = LinearInterpolate(A=@n180.R, B=LinearInterpolate(A=@n180.B, B=@n180.B, Alpha=@n181.G), Alpha=@n181.R)
  NamedRerouteDeclaration(None=MaterialFunctionCall(In=LinearInterpolate(A=@n182, B=@n180.G, Alpha=@n181.B), Contrast=@n78) {materialFunction=&CheapContrast}.Result) {name=Tri_Mask, nodeColor=<rgba 1 0.4688 0 1>}
  n183 = VectorParameter() {group=Wrap_Mitosis_Corruption, parameterName=Mitosis_NoiseB_TileRate, defaultValue=<rgba 1 3 3 2.5>}
  n184 = NamedRerouteUsage()
  n185 = Multiply(A=ComponentMask(None=Multiply(A=@n183.RGB, B=@n183.RGB)) {g=true, r=true}, B=@n184)
  NamedRerouteDeclaration(None=@n185) {name=Noise_B_Tilerate, nodeColor=<rgba 0.5469 1 0 1>}
  n186 = ScalarParameter() {defaultValue=100, group=Wrap_Mitosis_Corruption_UVS, parameterName=Mitosis_Scale_Noise_Vehicle, sortPriority=200}
  n187 = ScalarParameter() {defaultValue=100, group=Wrap_Mitosis_Corruption_UVS, parameterName=Mitosis_Scale_Noise_Vehicle_Side, sortPriority=200}
  n188 = MaterialFunctionCall(Float4=NamedRerouteUsage()) {materialFunction=&BreakOutFloat4Components}
  n189 = LinearInterpolate(
   A=LinearInterpolate(
    A=ScalarParameter() {defaultValue=50, group=Wrap_Mitosis_Corruption_UVS, parameterName=Mitosis_Scale_Noise_Weapon, sortPriority=200}
    B=LinearInterpolate(A=@n186, B=@n187, Alpha=@n188.G)
    Alpha=@n188.R
   )
   B=LinearInterpolate(
    A=ScalarParameter() {defaultValue=100, group=Wrap_Mitosis_Corruption_UVS, parameterName=Mitosis_Scale_Noise_Mech, sortPriority=200}
    B=ScalarParameter() {defaultValue=100, group=Wrap_Mitosis_Corruption_UVS, parameterName=Mitosis_Scale_Noise_Character, sortPriority=200}
    Alpha=@n188.A
   )
   Alpha=@n188.B
  )
  n190 = ScalarParameter() {defaultValue=100, group=Wrap_Mitosis_Corruption_UVS, parameterName=Mitosis_Scale_Noise_Backpack, sortPriority=200}
  n191 = MaterialFunctionCall(Float2=NamedRerouteUsage()) {materialFunction=&BreakOutFloat2Components}
  NamedRerouteDeclaration(None=LinearInterpolate(
    A=LinearInterpolate(A=@n189, B=@n190, Alpha=@n191.R)
    B=ScalarParameter() {defaultValue=100, group=Wrap_Mitosis_Corruption_UVS, parameterName=Mitosis_Scale_Noise_Pickaxe, sortPriority=200}
    Alpha=@n191.G
   )) {name=UVs_Scale_Noise, nodeColor=<rgba 0 0.7188 1 1>}
  n192 = VectorParameter() {group=Wrap_Mitosis_Corruption, parameterName=Mitosis_NoiseC_TileRate, defaultValue=<rgba 0.25 3 3 1>}
  n193 = Multiply(A=@n184, B=ComponentMask(None=Multiply(A=@n192.RGB, B=@n192.RGB)) {g=true, r=true})
  NamedRerouteDeclaration(None=@n193) {name=Noise_C_Tilerate, nodeColor=<rgba 0.5469 1 0 1>}
  n194 = VectorParameter() {group=Wrap_Mitosis_Corruption, parameterName=Mitosis_NoiseC_Speed, defaultValue=<rgba -0.025 -0.0031 0 1>}
  n195 = ComponentMask(None=Multiply(A=Time(), B=@n194.RGB)) {g=true, r=true}
  NamedRerouteDeclaration(None=@n195) {name=Noise_C_Tilerate_Speed, nodeColor=<rgba 0.5469 1 0 1>}
  n196 = ComponentMask(None=Multiply(A=Time(), B=VectorParameter() {group=Wrap_Mitosis_Corruption, parameterName=Mitosis_NoiseB_Speed, defaultValue=<rgba 0.015 0.003 0 1>}.RGB)) {g=true, r=true}
  NamedRerouteDeclaration(None=@n196) {name=Noise_B_Tilerate_Speed, nodeColor=<rgba 0.5469 1 0 1>}
  n197 = Constant3Vector() {constant=<rgba -0.25 0 0 1>}
  n198 = Constant3Vector() {constant=<rgba 0 0 0 1>}
  n199 = MaterialFunctionCall(Float4=NamedRerouteUsage()) {materialFunction=&BreakOutFloat4Components}
  n200 = LinearInterpolate(
   A=LinearInterpolate(A=Constant3Vector() {constant=<rgba 0 -0.25 -0.25 1>}, B=LinearInterpolate(A=@n197, B=@n198, Alpha=@n199.G), Alpha=@n199.R)
   B=LinearInterpolate(A=Constant3Vector() {constant=<rgba -0.25 -0.25 0.25 1>}, B=Constant3Vector() {constant=<rgba -0.25 -0.25 0.25 1>}, Alpha=@n199.A)
   Alpha=@n199.B
  )
  n201 = Constant3Vector() {constant=<rgba -0.25 -0.25 0 1>}
  n202 = MaterialFunctionCall(Float2=NamedRerouteUsage()) {materialFunction=&BreakOutFloat2Components}
  NamedRerouteDeclaration(None=LinearInterpolate(A=LinearInterpolate(A=@n200, B=@n201, Alpha=@n202.R), B=Constant3Vector() {constant=<rgba 0 0 0 1>}, Alpha=@n202.G)) {name=tri_Rotation, nodeColor=<rgba 0 0.2969 1 1>}
  n203 = TextureObjectParameter() {
   group=Wrap_Mitosis_Corruption
   parameterName=Mitosis__Tex_Noise
   samplerType=SAMPLERTYPE_LinearColor
   texture=/Game/Gadgets/Assets/VinderTech_GliderChute/Glider_Mitosis_Male/Textures/Noise/VT_GaussianNoise
  }
  NamedRerouteDeclaration(None=@n203) {name=Tex_Noise, nodeColor=<rgba 0.1953 1 0 1>}
  NamedRerouteDeclaration(None=@n50) {name=Dot_outpot, nodeColor=<rgba 1 0.3984 0 1>}
  n204 = VectorParameter() {group=Wrap_Mitosis_Corruption_UVS, parameterName=Mitosis_Noise_Speed_Weapon, defaultValue=<rgba 0 8 -5 1>}
  n205 = VectorParameter() {group=Wrap_Mitosis_Corruption_UVS, parameterName=Mitosis_Noise_Speed_Vehicle, defaultValue=<rgba 0 0 0 1>}
  n206 = VectorParameter() {group=Wrap_Mitosis_Corruption_UVS, parameterName=Mitosis_Noise_Speed_VehicleSide, defaultValue=<rgba 0 0 -1 1>}
  n207 = MaterialFunctionCall(Float4=NamedRerouteUsage()) {materialFunction=&BreakOutFloat4Components}
  n208 = VectorParameter() {group=Wrap_Mitosis_Corruption_UVS, parameterName=Mitosis_Noise_Speed_Character, defaultValue=<rgba 0 0 0 1>}
  n209 = LinearInterpolate(
   A=LinearInterpolate(A=@n204.RGB, B=LinearInterpolate(A=@n205.RGB, B=@n206.RGB, Alpha=@n207.G), Alpha=@n207.R)
   B=LinearInterpolate(
    A=VectorParameter() {group=Wrap_Mitosis_Corruption_UVS, parameterName=Mitosis_Noise_Speed_Mech, defaultValue=<rgba 0 0 0 1>}.RGB
    B=@n208.RGB
    Alpha=@n207.A
   )
   Alpha=@n207.B
  )
  n210 = VectorParameter() {group=Wrap_Mitosis_Corruption_UVS, parameterName=Mitosis_Noise_Speed_Backpack, defaultValue=<rgba 0 0 0 1>}
  n211 = MaterialFunctionCall(Float2=NamedRerouteUsage()) {materialFunction=&BreakOutFloat2Components}
  n212 = VectorParameter() {group=Wrap_Mitosis_Corruption_UVS, parameterName=Mitosis_Noise_Speed_Pickaxe, defaultValue=<rgba 0 0 0 1>}
  NamedRerouteDeclaration(None=LinearInterpolate(A=LinearInterpolate(A=@n209, B=@n210.RGB, Alpha=@n211.R), B=@n212.RGB, Alpha=@n211.G)) {name=UVs_Noise_Speed, nodeColor=<rgba 0 0.625 1 1>}
  NamedRerouteDeclaration(None=FunctionInput(Preview=Constant4Vector() {constant=<rgba 1 0 0 0>}.RGBA) {bUsePreviewValueAsDefault=true, inputName=IsState, inputType=FunctionInput_Vector4, sortPriority=4}) {name=IsState, nodeColor=<rgba 0.1497 1 0.1633 1>}
  n213 = TextureObjectParameter() {
   group=Wrap_Mitosis
   parameterName=Mitosis__Tex_N2
   samplerType=SAMPLERTYPE_Normal
   texture=&FlatNormalBlue
  }
  NamedRerouteDeclaration(None=@n213) {name=Tex_Nrml_1, nodeColor=<rgba 0.1953 1 0 1>}
  NamedRerouteDeclaration(None=LinearInterpolate(
    A=@n69.Raw_Normal
    B=MaterialFunctionCall!(
     Texture_Normals=@n33
     Rotation_Axis=@n72
     Caculated_Bounds=@n75
     Pre_Skinned_Normals=@n76
     Blend_Contrast=@n78
     Fill_alternate_Axis_Opacity=@n77
     Is_Vehicle=@n68.R
     Is_Mech=@n68.R
    ) {materialFunction=&/MF_TopsSideProjection_Normals}.Add_Surface_Normal
    Alpha=@n81
   )) {name=Normal_Curption_veins, nodeColor=<rgba 0 1 0.8125 1>}
  n214 = ScalarParameter() {defaultValue=100, group=Wrap_Mitosis_Corruption_UVS, parameterName=Mitosis_Scale_WrapNoise_Vehicle, sortPriority=200}
  n215 = ScalarParameter() {defaultValue=100, group=Wrap_Mitosis_Corruption_UVS, parameterName=Mitosis_Scale_WrapNoise_Vehicle_Side, sortPriority=200}
  n216 = MaterialFunctionCall(Float4=NamedRerouteUsage()) {materialFunction=&BreakOutFloat4Components}
  n217 = LinearInterpolate(
   A=LinearInterpolate(
    A=ScalarParameter() {defaultValue=50, group=Wrap_Mitosis_Corruption_UVS, parameterName=Mitosis_Scale_WrapNoise_Weapon, sortPriority=200}
    B=LinearInterpolate(A=@n214, B=@n215, Alpha=@n216.G)
    Alpha=@n216.R
   )
   B=LinearInterpolate(
    A=ScalarParameter() {defaultValue=100, group=Wrap_Mitosis_Corruption_UVS, parameterName=Mitosis_Scale_WrapNoise_Mech, sortPriority=200}
    B=ScalarParameter() {defaultValue=100, group=Wrap_Mitosis_Corruption_UVS, parameterName=Mitosis_Scale_WrapNoise_Character, sortPriority=200}
    Alpha=@n216.A
   )
   Alpha=@n216.B
  )
  n218 = ScalarParameter() {defaultValue=100, group=Wrap_Mitosis_Corruption_UVS, parameterName=Mitosis_Scale_WrapNoise_Backpack, sortPriority=200}
  n219 = MaterialFunctionCall(Float2=NamedRerouteUsage()) {materialFunction=&BreakOutFloat2Components}
  NamedRerouteDeclaration(None=LinearInterpolate(
    A=LinearInterpolate(A=@n217, B=@n218, Alpha=@n219.R)
    B=ScalarParameter() {defaultValue=100, group=Wrap_Mitosis_Corruption_UVS, parameterName=Mitosis_Scale_WrapNoise_Pickaxe, sortPriority=200}
    Alpha=@n219.G
   )) {name=UVs_Scale_Noise_Warp, nodeColor=<rgba 0 0.7188 1 1>}
  n220 = ChannelMaskParameter(None=@n82) {group=Wrap_Mitosis_Corruption, maskChannel=Blue, parameterName=Mitosis_Glow_Mask_Channel, defaultValue=<rgba 0 0 1 0>}
  NamedRerouteDeclaration(None=@n220.R) {name=Corruption Glow_Msk_1, nodeColor=<rgba 0 1 0.2031 1>}
  n221 = Frac(None=Add(A=Constant(), B=Multiply(A=@n51) {constB=0.5}))
  n222 = CurveAtlasRowParameter(CurveTime=@n221) {
   atlas=&CruveAtlas_Wraps_01
   curve=&CC_Mitos_Coruption_Color
   defaultValue=84
   group=Wrap_Mitosis_Corruption
   parameterName=Mitosis_Corruption_Colors_inner_CC
   primitiveDataIndex=1
   sortPriority=0
  }
  NamedRerouteDeclaration(None=MaterialFunctionCall(Index=@n42, CCinput=@n222, UV=@n221) {materialFunction=&MF_Global_CC_Tex}.3V) {name=Corruption _Colors_Front, nodeColor=<rgba 0 1 0.4141 1>}
  NamedRerouteDeclaration(None=FunctionInput(Preview=Constant2Vector()) {bUsePreviewValueAsDefault=true, inputName=IsState_B, inputType=FunctionInput_Vector2, sortPriority=4}) {name=IsState_B, nodeColor=<rgba 0.1497 1 0.1633 1>}
  n223 = PreSkinnedLocalBounds()
  n224 = Subtract(A=VertexInterpolator(VS=PreSkinnedPosition()).PS, B=@n223.Min)
  n225 = MaterialFunctionCall(Color=Subtract(A=@n223.Max, B=@n223.Max)) {materialFunction=&SplitComponents}
  n226 = Max(A=Max(A=@n225.R, B=@n225.R), B=@n225.B)
  NamedRerouteDeclaration(None=FunctionInput(Preview=Divide(A=@n224, B=@n226)) {bUsePreviewValueAsDefault=true, inputName=Calcated_Bounds, sortPriority=11}) {name=Caclcated_Bounds, nodeColor=<rgba 0 0.9297 1 1>}
  n227 = Clamp(None=DotProduct(A=CameraVectorWS(), B=PixelNormalWS()))
  n228 = CurveAtlasRowParameter(CurveTime=@n227) {
   atlas=&CruveAtlas_Wraps_01
   curve=/Game/Weapons/WeaponSkins/Wraps/Textures/0_ColorCurves/CC_MetalicLUT
   defaultValue=76
   group=Wrap_Mitosis_Fresnel
   parameterName=Mitosis_Metal_LUT_CC
   primitiveDataIndex=1
   sortPriority=-5
  }
  n229 = ChannelMaskParameter(None=MaterialFunctionCall(
    Index=ScalarParameter() {defaultValue=76, group=Wrap_Mitosis_Fresnel, parameterName="Mitosis_Metal_LUT_CC  Index", sortPriority=-5}
    CCinput=@n228
    CCinput_Alpha=@n228
    UV=@n227
   ) {materialFunction=&MF_Global_CC_Tex}.4v) {group=Wrap_Mitosis_Fresnel, maskChannel=Green, parameterName=Mitosis_Metal_LUT_Channel, sortPriority=-5, defaultValue=<rgba 0 1 0 0>}
  NamedRerouteDeclaration(None=@n229.R) {name=LUT_Metal, nodeColor=<rgba 0 1 0.625 1>}
  n230 = PreSkinnedNormal()
  NamedRerouteDeclaration(None=FunctionInput(Preview=VertexInterpolator(VS=@n230).PS) {bUsePreviewValueAsDefault=true, inputName=Pre-Skinned Local Normal, sortPriority=10}) {name=PreSkin_Normal, nodeColor=<rgba 0.8281 1 0 1>}
  NamedRerouteDeclaration(None=@n30.Normal) {name=Root_Normal, nodeColor=<rgba 0.4062 0 1 1>}
  NamedRerouteUsage()
  NamedRerouteUsage()
  NamedRerouteUsage()
  OneMinus(None=@n122)
  ScalarParameter() {defaultValue=2, group=Wrap_Mitosis_Corruption, parameterName=Mitosis_NoiseC_TileRate}
  ScalarParameter() {defaultValue=150, group=Wrap_Mitosis_Corruption, parameterName=Mitosis_Noise_PatternScale}
  ScalarParameter() {defaultValue=100, group=Wrap_Mitosis_Corruption, parameterName=Mitosis_Noise_Added_size}
  SetMaterialAttributes(MaterialAttributes=@n108, BaseColor=Constant3Vector() {constant=<rgba 1 0 0 0>}) {attributeSetTypes [
    #2
    #3
    #5
    #4
    #1
    #6
   ]}
  n231 = NamedRerouteUsage()
  TextureSample(UVs=Add(A=Multiply(A=@n231, B=@n185), B=@n196), Tex=NamedRerouteUsage()) {samplerType=SAMPLERTYPE_LinearGrayscale, texture=&T_Wrap_Mitosis_MSk}
  n232 = NamedRerouteUsage()
  TextureSample(UVs=Add(A=Multiply(A=@n232, B=@n193), B=@n195), Tex=NamedRerouteUsage()) {samplerType=SAMPLERTYPE_LinearGrayscale, texture=&T_Wrap_Mitosis_MSk}
  VectorParameter() {group=Wrap_Mitosis_Corruption, parameterName=Mitosis_Noise_PatternDirection, defaultValue=<rgba 0 8 -10 0>}
  VertexNormalWS()
 ]}
