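# Wrap material function that layers an audio-reactive neon pulse look onto weapon, vehicle, mech or character materials, behind a static switch.
# Fresnel plus a radial mask on camera-facing virtual-plane UVs forms outer glow, joined on non-mobile paths by per-type Crevice edge masks.
# That mask drives metallic, roughness and time-scrolled gradient emissive.
# Event and emote audio shift gradient color and glow intensity before the result blends over the input attributes.
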MaterialFunction /Game/Weapons/WeaponSkins/Wraps/MaterialFunctions/MF_WrapNeonPulse
graph {nodes [
  n1 = PreSkinnedLocalBounds()
  n2 = Add(A=@n1.Min, B=FunctionInput() {inputName=Object Bounds Min Adjustment, sortPriority=51}) {constB=0}
  n3 = FunctionInput() {inputName=Object Bounds Max Adjustment, sortPriority=52}
  n4 = Multiply(A=Subtract(A=Add(A=@n1.Max, B=@n3) {constB=0}, B=@n2)) {constB=0.5}
  n5 = Add(A=@n2, B=@n4)
  Add(A=Divide(A=Subtract(B=@n2), B=@n4), B=Add(A=@n5) {constB=0})
  n6 = CollectionParameter() {collection=&FortniteMaterialParameters, parameterName=EventAudio_Spectrum}
  ComponentMask(None=@n6) {r=true}
  ComponentMask(None=@n6) {g=true}
  ComponentMask(None=@n6) {b=true}
  ComponentMask(None=@n6) {a=true}
  n7 = Constant()
  FunctionInput(Preview=@n7) {bUsePreviewValueAsDefault=true, inputName=Side Project Vehicle, inputType=FunctionInput_Scalar, sortPriority=10}
  n8 = FunctionInput() {inputName=Material_input, inputType=FunctionInput_MaterialAttributes, sortPriority=-10}
  n9 = Fresnel(
   ExponentIn=ScalarParameter() {defaultValue=0.5, group=Wrap_NeonPulse, parameterName=NeonP_Fresnel_ExponentIn, sortPriority=200}
   BaseReflectFractionIn=ScalarParameter() {defaultValue=0.04, group=Wrap_NeonPulse, parameterName=NeonP_Fesnel_BaseReflectFractionIn, sortPriority=200}
  )
  n10 = ScalarParameter() {group=Wrap_NeonPulse, parameterName=NeonP_OuterGlow_Strenght, sortPriority=200}
  n11 = FunctionInput(Preview=@n7) {bUsePreviewValueAsDefault=true, inputName=isMech, inputType=FunctionInput_Scalar, sortPriority=10}
  n12 = LinearInterpolate(A=TransformPosition(None=@n5) {transformType=TRANSFORMPOSSOURCE_World}, B=ObjectPositionWS(), Alpha=@n11)
  n13 = Normalize(VectorInput=Subtract(A=@n12, B=CameraPositionWS()))
  n14 = Constant3Vector() {constant=<rgba 0 0 1 1>}
  n15 = MaterialFunctionCall(Color=@n1.Extents) {materialFunction=&SplitComponents}
  n16 = Max(A=Max(A=@n15.R, B=@n15.R), B=@n15.B)
  n17 = MaterialFunctionCall() {materialFunction=/Engine/Functions/Engine_MaterialFunctions02/WorldPositionOffset/ObjectScale}
  n18 = ScalarParameter() {defaultValue=1, group=Wrap_NeonPulse, parameterName=NeonP_Weapon_RadialScale, sortPriority=200}
  n19 = ScalarParameter() {defaultValue=1, group=Wrap_NeonPulse, parameterName=NeonP_Vehicle_RadialScale, sortPriority=200}
  n20 = FunctionInput(Preview=Constant()) {inputName=isVehicle, inputType=FunctionInput_Scalar, sortPriority=10}
  n21 = MaterialFunctionCall!(
   Plane Normal Axis=@n13
   Plane X Axis=Normalize(VectorInput=CrossProduct(A=@n13, B=@n14))
   Plane Center=@n12
   UV Scale=Multiply(A=Multiply(A=@n16, B=@n17.Scale XYZ), B=LinearInterpolate(A=@n18, B=@n19, Alpha=@n20))
  ) {materialFunction=/Game/Athena/Environments/Nexus/Materials/VirtualPlaneCoordinatesAbs}
  n22 = NamedRerouteUsage()
  n23 = NamedRerouteUsage()
  n24 = OneMinus(None=SphereMask(
    A=LinearInterpolate(A=@n21.UVs, B=@n22, Alpha=@n23)
    B=Constant() {r=0.5}
    Radius=ScalarParameter() {defaultValue=0.5, group=Wrap_NeonPulse, parameterName=NeonP_Radius_A, sortPriority=200}
    Hardness=ScalarParameter() {defaultValue=-2, group=Wrap_NeonPulse, parameterName=NeonP_Hardness_A, sortPriority=200}
   ))
  n25 = Saturate(None=Add(A=@n9, B=Multiply(A=@n10, B=@n24)))
  n26 = Constant() {r=1}
  n27 = MaterialFunctionCall(
   Input=FunctionInput() {inputName=Crevice, inputType=FunctionInput_Scalar}
   Input Low=Constant() {r=0.5}
   Input High=@n26
   Target Low=Constant()
   Target High=@n26
  ) {materialFunction=&RemapValueRange}
  n28 = Power(Base=@n27.Result, Exp=ScalarParameter() {defaultValue=2, group=Wrap_NeonPulse, parameterName=NeonP_Weapon_EdgePower})
  n29 = ScalarParameter() {defaultValue=30, group=Wrap_NeonPulse, parameterName=NeonP_Weapon_EdgeStrenght}
  n30 = MaterialFunctionCall(
   In=Saturate(None=Multiply(A=@n28, B=@n29))
   Contrast=ScalarParameter() {defaultValue=0.5, group=Wrap_NeonPulse, parameterName=NeonP_Weapon_EdgeContrast}
  ) {materialFunction=&CheapContrast}
  n31 = Power(Base=@n27.Result, Exp=ScalarParameter() {defaultValue=2, group=Wrap_NeonPulse, parameterName=NeonP_Vehicle_EdgePower})
  n32 = ScalarParameter() {defaultValue=30, group=Wrap_NeonPulse, parameterName=NeonP_Vehicle_EdgeStrenght}
  n33 = MaterialFunctionCall(
   In=Saturate(None=Multiply(A=@n31, B=@n32))
   Contrast=ScalarParameter() {defaultValue=0.5, group=Wrap_NeonPulse, parameterName=NeonP_Vehicle_EdgeContrast}
  ) {materialFunction=&CheapContrast}
  n34 = ScalarParameter() {defaultValue=2, group=Wrap_NeonPulse, parameterName=NeonP_Mech_EdgePower}
  n35 = Saturate(None=Multiply(A=Power(Base=@n27.Result, Exp=@n34), B=ScalarParameter() {defaultValue=30, group=Wrap_NeonPulse, parameterName=NeonP_Mech_EdgeStrenght}))
  n36 = ScalarParameter() {defaultValue=0.5, group=Wrap_NeonPulse, parameterName=NeonP_Mech_EdgeContrast}
  n37 = Add(
   A=Multiply(A=@n25, B=ScalarParameter() {group=Wrap_NeonPulse, parameterName=NeonP_OuterGlow_Strenght_PC, sortPriority=200})
   B=LinearInterpolate(
    A=LinearInterpolate(A=@n30.Result, B=@n33.Result, Alpha=@n20)
    B=MaterialFunctionCall(In=@n35, Contrast=@n36) {materialFunction=&CheapContrast}.Result
    Alpha=@n11
   )
  )
  n38 = ShadingPathSwitch(Default=@n37, Deferred=@n37, Mobile=@n25) {desc=Migrated from FeatureLevelSwitch}
  n39 = Add(A=TextureCoordinate() {uTiling=0, vTiling=0}, B=ScalarParameter() {group=00-Reactive-GTFB, parameterName=00-Color_Shift})
  n40 = Multiply(A=Time(), B=ScalarParameter() {group=Wrap_NeonPulse, parameterName=NeonP_Color_Speed})
  n41 = ScalarParameter() {group=Wrap_NeonPulse, parameterName=NeonP_Spec_color_offset B, sortPriority=151}
  n42 = CollectionParameter() {collection=&FortniteMaterialParameters, parameterName=EventAudio_Amplitude}
  n43 = CollectionParameter() {collection=&FortniteMaterialParameters, parameterName=Emote_AudioLow}
  n44 = MaterialFunctionCall(
   Input=Max(A=@n42, B=@n43)
   Input Low=ScalarParameter() {group=Wrap_NeonPulse, parameterName=NeonP_Sectrum_Low_Remap, sortPriority=150}
   Input High=ScalarParameter() {defaultValue=1, group=Wrap_NeonPulse, parameterName=NeonP_Sectrum_High_Remap, sortPriority=150}
   Target Low=Constant()
   Target High=Constant() {r=1}
  ) {materialFunction=&RemapValueRange}
  n45 = Round(None=@n44.Result)
  n46 = TextureSampleParameter2D(UVs=Add(A=Add(A=@n39, B=@n40), B=LinearInterpolate(B=@n41, Alpha=@n45) {constB=0.3})) {
   group=Wrap_NeonPulse
   parameterName=NeonP_Color_Grad
   texture=/Game/Global/Textures/Default/Masks_and_Gradients/WeaponGradients/T_Wrap_Spectrum_Grad
  }
  n47 = ScalarParameter() {defaultValue=1, group=Wrap_NeonPulse, parameterName=NeonP_Color_Glow_min}
  n48 = ScalarParameter() {defaultValue=1, group=Wrap_NeonPulse, parameterName=NeonP_Color_Glow_Max}
  n49 = CollectionParameter() {collection=&FortniteMaterialParameters, parameterName=Emote_AudioHigh}
  n50 = Max(A=Saturate(None=@n44.Result), B=@n49)
  n51 = SetMaterialAttributes!(
   MaterialAttributes=GetMaterialAttributes(None=@n8) {attributeGetTypes [#1]}.MaterialAttributes
   BaseColor=VectorParameter() {group=Wrap_NeonPulse, parameterName=NeonP_Main_Color, defaultValue=<rgba 0 0 0 1>}.RGB
   Metallic=Saturate(None=OneMinus(None=@n38))
   EmissiveColor=Multiply(A=Multiply(A=@n38, B=@n46.RGB), B=LinearInterpolate(A=@n47, B=@n48, Alpha=@n50))
   Roughness=LinearInterpolate(
    A=ScalarParameter() {defaultValue=0.3, group=Wrap_NeonPulse, parameterName=NeonP_Roughnes_Base}
    B=ScalarParameter() {defaultValue=0.3, group=Wrap_NeonPulse, parameterName=NeonP_Roughnes_Edge}
    Alpha=@n38
   )
  ) {attributeSetTypes [
    #2
    #3
    #4
    #5
   ]}
  n52 = LinearInterpolate(
   B=FunctionInput() {bUsePreviewValueAsDefault=true, inputName=Customization Mask, inputType=FunctionInput_Scalar, sortPriority=24}
   Alpha=NamedRerouteUsage()
  ) {constA=1}
  FunctionOutput(None=StaticSwitchParameter(True=BlendMaterialAttributes(A=@n8, B=@n51, Alpha=@n52), False=@n8) {group=Customization_Wrap, parameterName=Wrap_NeonPulse})
  LinearInterpolate(A=@n25, B=@n37, Alpha=ScalarParameter() {group=Wrap_NeonPulse, parameterName=NeonP_Temp_mobile})
  n53 = TransformPosition(None=Constant3Vector()) {transformType=TRANSFORMPOSSOURCE_World}
  n54 = Normalize(VectorInput=Subtract(A=@n53, B=CameraPositionWS()))
  n55 = Constant3Vector() {constant=<rgba 0 0 1 1>}
  n56 = FunctionInput() {bUsePreviewValueAsDefault=true, inputName=Object-SafeZone-Virtual-Pivot, inputType=FunctionInput_Vector4, sortPriority=105}
  n57 = MaterialFunctionCall() {materialFunction=&ObjectScale}
  n58 = MaterialFunctionCall(Float3=PreSkinnedLocalBounds().Extents) {materialFunction=&BreakOutFloat3Components}
  n59 = Max(A=Max(A=@n58.R, B=@n58.R), B=@n58.B)
  MaterialFunctionCall(
   Plane Normal Axis=@n54
   Plane X Axis=Normalize(VectorInput=CrossProduct(A=@n54, B=@n55))
   Plane Center=Add(A=@n53, B=ComponentMask(None=@n56) {b=true, g=true, r=true})
   UV Scale=Multiply(A=ComponentMask(None=@n56) {a=true}, B=Multiply(A=@n57.Scale X, B=@n59))
  ) {materialFunction=&VirtualPlaneCoordinatesAbs}
  NamedRerouteDeclaration(None=FunctionInput() {bUsePreviewValueAsDefault=true, inputName=isCharacter, inputType=FunctionInput_Scalar, sortPriority=10}) {name=IsCharacter, nodeColor=<rgba 0.7377 1 0 1>}
  n60 = FunctionInput() {bUsePreviewValueAsDefault=true, inputName=Object-SafeZone-Virtual-CH-Switch, inputType=FunctionInput_Scalar, sortPriority=105}
  NamedRerouteDeclaration(None=MaterialFunctionCall!(
    Camera _pos
    Axis
    Scale
    Scale_Multplyer
    Object-SafeZone-Virtual-Pivot=@n56
    Object-SafeZone-Virtual-CH-Switch=@n60
    SinglePlane Normal Axis
    SinglePlane X Axis
    PlaneCenter_distortion
   ) {materialFunction=&/MF_SafeZone_VirtualPlane}.Uvs) {name=SafeZoneVP_UVs, nodeColor=<rgba 1 0 0.6562 1>}
  OneMinus(None=@n9)
 ]}
