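# Material function for the RainbowGlitter wrap cosmetic, blended over incoming material attributes by a customization mask behind a static switch.
# Bounds-normalized local positions become planar UVs, with projection, offsets and motion chosen by weapon, vehicle or mech inputs.
# A world-aligned sparkle texture, height-blended with a moving gradient wipe, drives glitter emissive.
# A Blueprint-set game time temporarily swaps glitter for fading gradient-colored rainbow and two-layer scrolling rain emissive, darkening masked base color.
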MaterialFunction /Game/Weapons/WeaponSkins/Wraps/MaterialFunctions/MF_WrapRainbowGlitter
graph {nodes [
  n1 = Constant3Vector() {constant=<rgba 1 1 -1 0>}
  n2 = PreSkinnedLocalBounds()
  n3 = VertexInterpolator(VS=PreSkinnedPosition())
  n4 = FunctionInput(Preview=Divide(A=Subtract(A=@n3.PS, B=@n2.Min), B=Subtract(A=@n2.Max, B=@n2.Max))) {bUsePreviewValueAsDefault=true, inputName=Cal_Bounds, sortPriority=10}
  n5 = VectorParameter() {group=Wrap_RainbowGlitter, parameterName=WRG_TextureOffset_Weapon, sortPriority=15, defaultValue=<rgba 0 0 0 1>}
  n6 = VectorParameter() {group=Wrap_RainbowGlitter, parameterName=WRG_TextureOffset_Vehicle, sortPriority=15, defaultValue=<rgba 0 0 0 1>}
  n7 = VectorParameter() {group=Wrap_RainbowGlitter, parameterName=WRG_TextureOffset_Vehicle_Side, sortPriority=15, defaultValue=<rgba 0 0 0 1>}
  n8 = Constant() {r=1}
  n9 = FunctionInput(Preview=@n8) {bUsePreviewValueAsDefault=true, inputName=Side Project Vehicle, inputType=FunctionInput_Scalar, sortPriority=20}
  n10 = FunctionInput(Preview=@n8) {inputName=isVehicle, inputType=FunctionInput_Scalar, sortPriority=20}
  n11 = LinearInterpolate(A=@n5.RGB, B=LinearInterpolate(A=@n6.RGB, B=@n7.RGB, Alpha=@n9), Alpha=@n10)
  n12 = VectorParameter() {group=Wrap_RainbowGlitter, parameterName=WRG_TextureOffset_Mech, sortPriority=15, defaultValue=<rgba 0 0 0 1>}
  n13 = FunctionInput(Preview=@n8) {bUsePreviewValueAsDefault=true, inputName=isMech, inputType=FunctionInput_Scalar, sortPriority=20}
  n14 = MaterialFunctionCall(Color=Add(A=Multiply(A=@n4, B=@n1), B=LinearInterpolate(A=@n11, B=@n12.RGB, Alpha=@n13))) {materialFunction=&SplitComponents}
  n15 = AppendVector(A=@n14.R, B=@n14.R)
  n16 = AppendVector(A=@n14.R, B=@n14.R)
  n17 = LinearInterpolate(
   A=LinearInterpolate(A=AppendVector(A=@n14.G, B=@n14.G), B=LinearInterpolate(A=@n15, B=@n16, Alpha=@n9), Alpha=@n10)
   B=AppendVector(A=OneMinus(None=@n14.B), B=@n14.R)
   Alpha=@n13
  )
  n18 = TextureObjectParameter() {
   group=Wrap_RainbowGlitter
   parameterName=WRG_Texture_M_01
   samplerType=SAMPLERTYPE_Masks
   sortPriority=10
   texture=/Game/Weapons/WeaponSkins/Wraps/Textures/RainbowGlitter/T_WRap_RainbowGlitter_M
  }
  n19 = TextureSample(UVs=@n17, Tex=@n18) {texture=/Game/Weapons/WeaponSkins/Wraps/Textures/RainbowGlitter/T_WRap_RainbowGlitter_BC}
  n20 = Add(
   A=Multiply(A=Time(), B=ScalarParameter() {group=Wrap_RainbowGlitter, parameterName=WRG_Rain_Color_Speed, sortPriority=125})
   B=Divide(A=@n19.G, B=ScalarParameter() {defaultValue=1, group=Wrap_RainbowGlitter, parameterName=WRG_Rain_Gradent_Spread, sortPriority=125})
  )
  Frac(None=@n20)
  n21 = Add(A=@n19.A, B=Multiply(A=Time(), B=ScalarParameter() {group=Wrap_RainbowGlitter, parameterName=WRG_RainBow_Color_Speed, sortPriority=125}))
  Frac(None=@n21)
  n22 = FunctionInput() {inputName=Material, inputType=FunctionInput_MaterialAttributes, sortPriority=-10}
  n23 = GetMaterialAttributes(None=@n22) {attributeGetTypes [#1, #2, #3]}
  n24 = Time()
  n25 = ScalarParameter() {defaultValue=-9999999, group=00-Reactive-GTFB, parameterName=00-GameTimeFromBP, sortPriority=150}
  n26 = Saturate(None=Divide(
    A=Subtract(A=@n24, B=@n25)
    B=ScalarParameter() {defaultValue=0.25, group=Wrap_RainbowGlitter, parameterName=WRG_Game_Duration, sortPriority=150}
   ))
  n27 = Add(A=OneMinus(None=@n19.R), B=@n26)
  n28 = TextureSampleParameter2D(UVs=@n17) {
   group=Wrap_RainbowGlitter
   parameterName=WRG_Texture_BC_01
   sortPriority=10
   texture=&T_WRap_RainbowGlitter_BC
  }
  n29 = TextureObject() {samplerType=SAMPLERTYPE_Masks, texture=/Game/Weapons/WeaponSkins/Wraps/Textures/RainbowGlitter/T_WRap_RainbowGlitter_M_1}
  n30 = ScalarParameter() {defaultValue=32, group=Wrap_RainbowGlitter, parameterName=WRG_Glitter_tile_size_Weapon, sortPriority=20}
  n31 = ScalarParameter() {defaultValue=32, group=Wrap_RainbowGlitter, parameterName=WRG_Glitter_tile_size_Vehicle, sortPriority=20}
  n32 = ScalarParameter() {defaultValue=32, group=Wrap_RainbowGlitter, parameterName=WRG_Glitter_tile_size_Mech, sortPriority=20}
  n33 = ScalarParameter() {defaultValue=24, group=Wrap_RainbowGlitter, parameterName=WRG_Glitter_tile_size_Character, sortPriority=20}
  n34 = NamedRerouteUsage()
  n35 = FunctionInput(Preview=VertexInterpolator(VS=PreSkinnedNormal()).PS) {bUsePreviewValueAsDefault=true, inputName=Pre-Skinned Local Normal, sortPriority=10}
  n36 = MaterialFunctionCall(
   Export Float 4=StaticBool() {value=true}
   TextureObject=@n29
   TextureSize=LinearInterpolate(A=LinearInterpolate(A=@n30, B=@n31, Alpha=@n10), B=LinearInterpolate(A=@n32, B=@n33, Alpha=@n34), Alpha=@n13)
   WorldPosition=FunctionInput(Preview=@n3.PS) {bUsePreviewValueAsDefault=true, inputName=Local Position, sortPriority=10}
   World Space Normal=@n35
  ) {materialFunction=&/WorldAlignedTexture}
  n37 = FunctionInput(Preview=Constant3Vector() {constant=<rgba 0 1 0 0>}) {bUsePreviewValueAsDefault=true, inputName=ForwardAxis, sortPriority=10}
  n38 = ScalarParameter() {defaultValue=0.2, group=Wrap_RainbowGlitter, parameterName=WRG_GlitterWipe_Speed_Weapon, sortPriority=35}
  n39 = ScalarParameter() {group=Wrap_RainbowGlitter, parameterName=WRG_GlitterWipe_Speed_Vehicle, sortPriority=35}
  n40 = Multiply(
   A=Time()
   B=LinearInterpolate(
    A=LinearInterpolate(A=@n38, B=@n39, Alpha=@n10)
    B=ScalarParameter() {group=Wrap_RainbowGlitter, parameterName=WRG_GlitterWipe_Speed_Mech, sortPriority=35}
    Alpha=@n13
   )
  )
  n41 = Add(
   A=MaterialFunctionCall!(Bounds=@n4, ForwardAxis=@n37, Offset=Constant(), FallOff=Constant() {r=4}) {materialFunction=&/MF_Wrap_Gradent_Simple}.Raw_Data
   B=@n40
  )
  n42 = Saturate(None=Multiply(A=Frac(None=@n41), B=ScalarParameter() {defaultValue=15, parameterName=B}) {constB=15})
  n43 = ScalarParameter() {defaultValue=4, group=Wrap_RainbowGlitter, parameterName=WRG_GlitterWipe_Falloff_Weapon, sortPriority=30}
  n44 = ScalarParameter() {group=Wrap_RainbowGlitter, parameterName=WRG_GlitterWipe_Falloff_Vehicle, sortPriority=30}
  n45 = MaterialFunctionCall(
   Bounds=@n4
   ForwardAxis=@n37
   Offset=Constant()
   FallOff=LinearInterpolate(
    A=LinearInterpolate(A=@n43, B=@n44, Alpha=@n10)
    B=ScalarParameter() {group=Wrap_RainbowGlitter, parameterName=WRG_GlitterWipe_Falloff_Mech, sortPriority=30}
    Alpha=@n13
   )
  ) {materialFunction=&MF_Wrap_Gradent_Simple}
  n46 = OneMinus(None=Frac(None=Add(A=@n45.Raw_Data, B=@n40)))
  n47 = Multiply(
   A=Add(
    A=Multiply(A=@n42, B=@n46)
    B=ScalarParameter() {defaultValue=-0.7, group=Wrap_RainbowGlitter, parameterName=WRG_GlitterWipe_Grad_Contrast, sortPriority=25}
   )
   B=ScalarParameter() {defaultValue=2.5, group=Wrap_RainbowGlitter, parameterName=WRG_GlitterWipe_Grad_Strenght, sortPriority=26}
  )
  n48 = Saturate(None=MaterialFunctionCall(
    Transition Phase=Saturate(None=@n47)
    Height Texture=ComponentMask(None=@n36.XYZ Texture) {r=true}
    Contrast=ScalarParameter() {defaultValue=10, group=Wrap_RainbowGlitter, parameterName=WRG_GlitterWipe_Grad_Sparkels Contrast, sortPriority=27}
   ) {materialFunction=&HeightLerp}.Alpha)
  n49 = Multiply(
   A=Multiply(A=@n28.RGB, B=Multiply(A=@n19.R, B=@n48))
   B=ScalarParameter() {defaultValue=1, group=Wrap_RainbowGlitter, parameterName=WRG_Glitter_Brightnes, sortPriority=20}
  )
  n50 = Floor(None=@n26)
  n51 = MaterialFunctionCall(Color=@n35) {materialFunction=&SplitComponents}
  n52 = ScalarParameter() {group=Wrap_RainbowGlitter, parameterName=WRG_Tri_Contrast, sortPriority=-10}
  n53 = MaterialFunctionCall(In=Abs(None=@n51.R), Contrast=@n52) {materialFunction=&CheapContrast}
  n54 = MaterialFunctionCall(In=Abs(None=@n51.G), Contrast=@n52) {materialFunction=&CheapContrast}
  n55 = LinearInterpolate(
   A=MaterialFunctionCall(In=Abs(None=@n51.B), Contrast=@n52) {materialFunction=&CheapContrast}.Result
   B=@n54.Result
   Alpha=@n9
  )
  n56 = Multiply(A=Multiply(A=@n4), B=ScalarParameter() {defaultValue=2, group=Wrap_RainbowGlitter, parameterName=WRG_RainB_TileRate, sortPriority=122})
  n57 = Multiply(A=@n56, B=Constant3Vector() {constant=<rgba 1 1 -1 0>})
  n58 = Multiply(A=@n19.B)
  n59 = LinearInterpolate(A=@n56, B=@n57, Alpha=@n58)
  n60 = Multiply(A=Constant3Vector() {constant=<rgba 1 -1 -1 0>}, B=@n56)
  n61 = LinearInterpolate(A=@n59, B=LinearInterpolate(A=LinearInterpolate(A=@n60, B=@n57, Alpha=@n58), B=@n59, Alpha=@n9), Alpha=@n10)
  n62 = LinearInterpolate(A=Multiply(A=@n56, B=Constant3Vector() {constant=<rgba -1 1 -1 0>}), B=@n57, Alpha=@n58)
  n63 = Time()
  n64 = VectorParameter() {group=Wrap_RainbowGlitter, parameterName=WRG_RainB_Weapon_Speed, sortPriority=120, defaultValue=<rgba 0 -0.2 -0.1 1>}
  n65 = VectorParameter() {group=Wrap_RainbowGlitter, parameterName=WRG_RainB_Vehicle_Speed, sortPriority=120, defaultValue=<rgba 0 0 0 1>}
  n66 = VectorParameter() {group=Wrap_RainbowGlitter, parameterName=WRG_RainB_Vehicle_Side_Speed, sortPriority=120, defaultValue=<rgba 0 0 0 1>}
  n67 = VectorParameter() {group=Wrap_RainbowGlitter, parameterName=WRG_RainB_Mech_Speed----------, sortPriority=120, defaultValue=<rgba 0 0 0 1>}
  n68 = LinearInterpolate(A=LinearInterpolate(A=@n64.RGB, B=LinearInterpolate(A=@n65.RGB, B=@n66.RGB, Alpha=@n9), Alpha=@n10), B=@n67.RGB, Alpha=@n13)
  n69 = MaterialFunctionCall(Color=Add(A=LinearInterpolate(A=@n61, B=@n62, Alpha=@n13), B=Multiply(A=@n63, B=@n68))) {materialFunction=&SplitComponents}
  n70 = AppendVector(A=@n69.R, B=@n69.R)
  n71 = AppendVector(A=@n69.R, B=@n69.R)
  n72 = LinearInterpolate(
   A=LinearInterpolate(A=AppendVector(A=@n69.G, B=@n69.G), B=LinearInterpolate(A=@n70, B=@n71, Alpha=@n9), Alpha=@n10)
   B=AppendVector(A=OneMinus(None=@n69.B), B=@n69.R)
   Alpha=@n13
  )
  n73 = Constant() {r=1}
  n74 = Multiply(A=MaterialFunctionCall(
    Transition Phase=@n19.G
    Height Texture=TextureSample(UVs=@n72, Tex=@n29) {samplerType=SAMPLERTYPE_Masks, texture=&T_WRap_RainbowGlitter_M_1}.A
    Contrast=@n73
   ) {materialFunction=&HeightLerp}.Alpha) {constB=0.5}
  n75 = Multiply(A=@n4, B=ScalarParameter() {defaultValue=2, group=Wrap_RainbowGlitter, parameterName=WRG_RainA_TileRate, sortPriority=102})
  n76 = Multiply(A=@n75, B=Constant3Vector() {constant=<rgba 1 1 -1 0>})
  n77 = LinearInterpolate(A=@n75, B=@n76, Alpha=@n19.B)
  n78 = Multiply(A=Constant3Vector() {constant=<rgba 1 -1 -1 0>}, B=@n75)
  n79 = LinearInterpolate(A=@n77, B=LinearInterpolate(A=LinearInterpolate(A=@n78, B=@n76, Alpha=@n19.B), B=@n77, Alpha=@n9), Alpha=@n10)
  n80 = LinearInterpolate(A=Multiply(A=@n75, B=Constant3Vector() {constant=<rgba -1 1 -1 0>}), B=@n76, Alpha=@n19.B)
  n81 = Time()
  n82 = VectorParameter() {group=Wrap_RainbowGlitter, parameterName=WRG_RainA_Weapon_Speed, sortPriority=100, defaultValue=<rgba 0 -0.2 -0.1 1>}
  n83 = VectorParameter() {group=Wrap_RainbowGlitter, parameterName=WRG_RainA_Vehicle_Speed, sortPriority=100, defaultValue=<rgba 0 0 0 1>}
  n84 = VectorParameter() {group=Wrap_RainbowGlitter, parameterName=WRG_RainA_Vehicle_Side_Speed, sortPriority=100, defaultValue=<rgba 0 0 0 1>}
  n85 = VectorParameter() {group=Wrap_RainbowGlitter, parameterName=WRG_RainA_Mech_Speed----------, sortPriority=100, defaultValue=<rgba 0 0 0 1>}
  n86 = LinearInterpolate(A=LinearInterpolate(A=@n82.RGB, B=LinearInterpolate(A=@n83.RGB, B=@n84.RGB, Alpha=@n9), Alpha=@n10), B=@n85.RGB, Alpha=@n13)
  n87 = MaterialFunctionCall(Color=Add(A=LinearInterpolate(A=@n79, B=@n80, Alpha=@n13), B=Multiply(A=@n81, B=@n86))) {materialFunction=&SplitComponents}
  n88 = LinearInterpolate(
   A=AppendVector(A=@n87.G, B=@n87.G)
   B=LinearInterpolate(A=AppendVector(A=@n87.R, B=@n87.R), B=AppendVector(A=@n87.R, B=@n87.R), Alpha=@n9)
   Alpha=@n10
  )
  n89 = AppendVector(A=OneMinus(None=@n87.B), B=@n87.R)
  n90 = MaterialFunctionCall(
   Transition Phase=@n19.G
   Height Texture=TextureSample(UVs=LinearInterpolate(A=@n88, B=@n89, Alpha=@n13), Tex=@n29) {samplerType=SAMPLERTYPE_Masks, texture=&T_WRap_RainbowGlitter_M_1}.A
   Contrast=@n73
  ) {materialFunction=&HeightLerp}
  n91 = Multiply(
   A=LinearInterpolate(A=LinearInterpolate(A=@n53.Result, B=@n55, Alpha=@n10), B=@n54.Result, Alpha=@n13)
   B=Saturate(None=Add(A=@n74, B=@n90.Alpha))
  )
  n92 = TextureObject() {texture=/Game/Weapons/WeaponSkins/Wraps/Textures/RainbowGlitter/T_WRap_RainbowGlitter_Grad}
  n93 = TextureSample(UVs=@n20, Tex=@n92) {texture=&T_WRap_RainbowGlitter_Grad}
  n94 = TextureSample(UVs=@n21, Tex=@n92) {texture=&T_WRap_RainbowGlitter_Grad}
  n95 = Add(
   A=Multiply(A=Multiply(A=@n91, B=@n93.RGB), B=ScalarParameter() {group=Wrap_RainbowGlitter, parameterName=WRG_Rain_Brightness, sortPriority=125})
   B=Multiply(
    A=Multiply(A=@n19.R, B=@n94.RGB)
    B=ScalarParameter() {defaultValue=1, group=Wrap_RainbowGlitter, parameterName=WRG_RainBow_Color_Brightness, sortPriority=125}
   )
  )
  n96 = OneMinus(None=Power(Base=@n26, Exp=ScalarParameter() {defaultValue=0.5, group=Wrap_RainbowGlitter, parameterName=WRG_Game_Curve, sortPriority=150}))
  n97 = SetMaterialAttributes!(
   MaterialAttributes=@n23.MaterialAttributes
   BaseColor=Multiply(A=@n28.RGB, B=Saturate(None=@n27))
   Roughness=LinearInterpolate(
    A=ScalarParameter() {group=Wrap_RainbowGlitter, parameterName=WRG_Roughness_Min, sortPriority=11}
    B=ScalarParameter() {defaultValue=1, group=Wrap_RainbowGlitter, parameterName=WRG_Roughness_Max, sortPriority=11}
    Alpha=@n28.A
   )
   Metallic=Constant()
   EmissiveColor=Add(A=LinearInterpolate(B=@n49, Alpha=@n50) {constB=0}, B=Multiply(A=@n95, B=@n96) {constB=0})
  ) {attributeSetTypes [
    #1
    #2
    #4
    #5
   ]}
  n98 = Saturate(None=Add(
    A=FunctionInput() {bUsePreviewValueAsDefault=true, inputName="Customization Mask (_CM B)", inputType=FunctionInput_Scalar, sortPriority=24}
    B=ScalarParameter() {group=Wrap_RainbowGlitter, parameterName=WRG_Remove_Customization Mask, sortPriority=-10}
   ))
  FunctionOutput(None=StaticSwitchParameter(True=BlendMaterialAttributes(A=@n23.MaterialAttributes, B=@n97, Alpha=@n98), False=@n22) {group=Customization_Wrap, parameterName=Wrap_RainbowGlitter})
  NamedRerouteDeclaration(None=FunctionInput() {bUsePreviewValueAsDefault=true, inputName=isCharacter, inputType=FunctionInput_Scalar, sortPriority=20}) {name=IsCharacter, nodeColor=<rgba 0.7344 1 0 1>}
  ScalarParameter() {group=Wrap_RainbowGlitter, parameterName=WRG_Animate_test, sortPriority=-10}
 ]}
