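# Wrap material function for the AlDay skin on weapons, vehicles and mechs, applied through customization masks when the Wrap_AlDay static switch is on.
# Bounds-based UVs, adjusted per object type, drive six independently placed, time-rotated gears from quadrant atlases via MF_ClockWork_Gears and MF_AlDay_Gears.
# A normal-based axis mask and gear masks composite them over a top/side-projected base texture.
# An optionally animated hue shift recolors the composite before output.
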MaterialFunction /Game/Weapons/WeaponSkins/Wraps/MaterialFunctions/MF_Wrap_AlDay
graph {nodes [
  n1 = FunctionInput() {inputName="Customization Mask (_CM B)", inputType=FunctionInput_Scalar, sortPriority=24, previewValue=<1 0 0 1>}
  n2 = FunctionInput() {inputName=Customization Mask B, inputType=FunctionInput_Scalar, sortPriority=24, previewValue=<1 0 0 1>}
  Add(A=@n1, B=@n2)
  n3 = VectorParameter() {group=Wrap_AlDay, parameterName=Ald_Acent_Colors, defaultValue=<rgba 0.314 0.0262 0.2051 1>}
  n4 = FunctionInput() {inputName=Material, inputType=FunctionInput_MaterialAttributes}
  n5 = GetMaterialAttributes(None=@n4) {attributeGetTypes [#1, #2, #3]}
  n6 = Constant3Vector() {constant=<rgba 0 0 1 0>}
  n7 = TextureObjectParameter() {group=Wrap_AlDay, parameterName=BC_Ald_01, sortPriority=-1, texture=&T_BlacksRGB}
  n8 = TextureObjectParameter() {
   group=Wrap_AlDay
   parameterName=N_Ald
   samplerType=SAMPLERTYPE_Normal
   sortPriority=-1
   texture=&FlatNormalBlue
  }
  n9 = ComponentMask(None=VectorParameter() {group=Wrap_AlDay, parameterName=Ald_C_Position, defaultValue=<rgba -0.375 -0.375 0 0>}.RGB) {g=true, r=true}
  n10 = FunctionInput(Preview=Constant()) {bUsePreviewValueAsDefault=true, inputName=isVehicle, inputType=FunctionInput_Scalar, sortPriority=4}
  n11 = FunctionInput(Preview=Constant()) {bUsePreviewValueAsDefault=true, inputName=isMech, inputType=FunctionInput_Scalar, sortPriority=4}
  n12 = PreSkinnedPosition()
  n13 = MaterialFunctionCall!(
   is Vehicle=@n10
   Object Bounds Max Adjustment=FunctionInput() {bUsePreviewValueAsDefault=true, inputName=Object Bounds Max Adjustment, sortPriority=15}
   Object Bounds Min Adjustment=FunctionInput() {bUsePreviewValueAsDefault=true, inputName=Object Bounds Min Adjustment, sortPriority=15}
   Object Bounds Center Adjustment=FunctionInput() {bUsePreviewValueAsDefault=true, inputName=Object Bounds Center Adjustment, sortPriority=15}
   isMech=@n11
   Side Project Vehicle
   Local Position=FunctionInput(Preview=VertexInterpolator(VS=@n12).PS) {bUsePreviewValueAsDefault=true, inputName=Local Position, sortPriority=2}
  ) {materialFunction=&/MF_WrapBounds}
  n14 = ComponentMask(None=@n13.Result) {g=true}
  n15 = ComponentMask(None=@n13.Result) {b=true}
  n16 = OneMinus(None=@n14)
  n17 = ComponentMask(None=@n13.Result) {r=true}
  n18 = FunctionInput() {bUsePreviewValueAsDefault=true, inputName=Side Project Vehicle, inputType=FunctionInput_Scalar, sortPriority=4}
  n19 = LinearInterpolate(A=AppendVector(A=@n14, B=@n15), B=LinearInterpolate(A=AppendVector(A=@n16, B=@n17), B=AppendVector(A=@n17, B=@n15), Alpha=@n18), Alpha=@n10)
  n20 = AppendVector(A=@n15, B=@n17)
  n21 = LinearInterpolate(
   A=ScalarParameter() {defaultValue=1, group=Wrap_AlDay, parameterName=TileRate_WeaponAld}
   B=ScalarParameter() {defaultValue=1, group=Wrap_AlDay, parameterName=TileRate_VehicleAld}
   Alpha=@n10
  )
  n22 = ScalarParameter() {defaultValue=1, group=Wrap_AlDay, parameterName=TileRate_MechAld}
  n23 = ComponentMask(None=VectorParameter() {group=Wrap_AlDay, parameterName=Offset_WeaponAld, defaultValue=<rgba 0 0 0 1>}.RGB) {g=true, r=true}
  n24 = LinearInterpolate(
   A=ComponentMask(None=VectorParameter() {group=Wrap_AlDay, parameterName=Offset_VehicleAld, defaultValue=<rgba 0 0 0 1>}.RGB) {g=true, r=true}
   B=ComponentMask(None=VectorParameter() {group=Wrap_AlDay, parameterName=Offset_VehicleSPAld, defaultValue=<rgba 0 0 0 1>}.RGB) {b=true, r=true}
   Alpha=@n18
  )
  n25 = VectorParameter() {group=Wrap_AlDay, parameterName=Offset_MechAld, defaultValue=<rgba 0 0 0 1>}
  n26 = Add(
   A=Multiply(A=LinearInterpolate(A=@n19, B=@n20, Alpha=@n11), B=LinearInterpolate(A=@n21, B=@n22, Alpha=@n11) {constA=1})
   B=LinearInterpolate(A=LinearInterpolate(A=@n23, B=@n24, Alpha=@n10), B=ComponentMask(None=@n25.RGB) {g=true, r=true}, Alpha=@n11)
  )
  n27 = MaterialFunctionCall(
   UVs=Multiply(A=Add(A=@n26, B=@n9)) {constB=4}
   Texture Scale=ScalarParameter() {defaultValue=1.8, group=Wrap_AlDay, parameterName=Ald_C_Scale}
  ) {materialFunction=&ScaleUVsByCenter}
  n28 = Frac(None=@n27.UVs)
  n29 = Constant2Vector() {r=0.5}
  n30 = Constant3Vector() {constant=<rgba 0.25 0.25 0 1>}
  n31 = MaterialFunctionCall(
   UVs=Add(A=Multiply(A=@n28) {constB=0.5}, B=@n29)
   Rotation Center=Add(A=@n29, B=ComponentMask(None=@n30) {g=true, r=true})
   "Rotation Angle (0-1)"=Multiply(A=Time(), B=ScalarParameter() {defaultValue=-0.1, group=Wrap_AlDay, parameterName=Ald_C_Rotation_Speed})
  ) {materialFunction=&CustomRotator}
  n32 = TextureObjectParameter() {
   desc=Red Gear mask Green Roughness
   group=Wrap_AlDay
   parameterName=Mask_Ald
   sortPriority=-1
   texture=&T_BlacksRGB
  }
  n33 = ComponentMask(None=VectorParameter() {group=Wrap_AlDay, parameterName=Ald_A_Position, defaultValue=<rgba -0.375 -0.375 0 0>}.RGB) {g=true, r=true}
  n34 = MaterialFunctionCall(
   UVs=Multiply(A=Add(A=@n26, B=@n33)) {constB=4}
   Texture Scale=ScalarParameter() {defaultValue=1.1, group=Wrap_AlDay, parameterName=Ald_A_Scale}
  ) {materialFunction=&ScaleUVsByCenter}
  n35 = Frac(None=@n34.UVs)
  n36 = Constant2Vector()
  n37 = Constant3Vector() {constant=<rgba 0.25 0.25 0 1>}
  n38 = MaterialFunctionCall(
   UVs=Add(A=Multiply(A=@n35) {constB=0.5}, B=@n36)
   Rotation Center=Add(A=@n36, B=ComponentMask(None=@n37) {g=true, r=true})
   "Rotation Angle (0-1)"=Multiply(A=Time(), B=ScalarParameter() {defaultValue=-0.12, group=Wrap_AlDay, parameterName=Ald_A_Rotation_Speed})
  ) {materialFunction=&CustomRotator}
  n39 = ComponentMask(None=VectorParameter() {group=Wrap_AlDay, parameterName=Ald_B_Position, defaultValue=<rgba -0.375 -0.375 0 0>}.RGB) {g=true, r=true}
  n40 = MaterialFunctionCall(
   UVs=Multiply(A=Add(A=@n26, B=@n39)) {constB=4}
   Texture Scale=ScalarParameter() {defaultValue=1.8, group=Wrap_AlDay, parameterName=Ald_B_Scale}
  ) {materialFunction=&ScaleUVsByCenter}
  n41 = Frac(None=@n40.UVs)
  n42 = Constant2Vector() {g=0.5}
  n43 = Constant3Vector() {constant=<rgba 0.25 0.25 0 1>}
  n44 = MaterialFunctionCall(
   UVs=Add(A=Multiply(A=@n41) {constB=0.5}, B=@n42)
   Rotation Center=Add(A=@n42, B=ComponentMask(None=@n43) {g=true, r=true})
   "Rotation Angle (0-1)"=Multiply(A=Time(), B=ScalarParameter() {defaultValue=0.09, group=Wrap_AlDay, parameterName=Ald_B_Rotation_Speed})
  ) {materialFunction=&CustomRotator}
  n45 = ComponentMask(None=VectorParameter() {group=Wrap_AlDay, parameterName=Ald_D_Position, defaultValue=<rgba -0.375 -0.375 0 0>}.RGB) {g=true, r=true}
  n46 = MaterialFunctionCall(
   UVs=Multiply(A=Add(A=@n26, B=@n45)) {constB=4}
   Texture Scale=ScalarParameter() {defaultValue=1.5, group=Wrap_AlDay, parameterName=Ald_D_Scale}
  ) {materialFunction=&ScaleUVsByCenter}
  n47 = Frac(None=@n46.UVs)
  n48 = Constant2Vector() {g=0.5, r=0.5}
  n49 = Constant3Vector() {constant=<rgba 0.25 0.25 0 1>}
  n50 = MaterialFunctionCall(
   UVs=Add(A=Multiply(A=@n47) {constB=0.5}, B=@n48)
   Rotation Center=Add(A=@n48, B=ComponentMask(None=@n49) {g=true, r=true})
   "Rotation Angle (0-1)"=Multiply(A=Time(), B=ScalarParameter() {defaultValue=0.15, group=Wrap_AlDay, parameterName=Ald_D_Rotation_Speed})
  ) {materialFunction=&CustomRotator}
  n51 = MaterialFunctionCall!(
   BackGroundMask=@n6
   Gear_mask_Red=@n32
   Gears=@n7
   Normals=@n8
   Uvs_GearA_Mask=@n34.UVs
   Uvs_GearA=@n38.Rotated Values
   Uvs_GearB_Mask=@n40.UVs
   Uvs_GearB=@n44.Rotated Values
   Uvs_GearC_Mask=@n27.UVs
   Uvs_GearC=@n31.Rotated Values
   Uvs_GearD_Mask=@n46.UVs
   Uvs_GearD=@n50.Rotated Values
   Local Normals
   IsState
   Rotation
  ) {materialFunction=&/MF_ClockWork_Gears}
  n52 = Constant()
  n53 = ScalarParameter() {defaultValue=0.5, group=Wrap_AlDay, parameterName=Ald_DecalRoughness, sortPriority=-1}
  n54 = SetMaterialAttributes!(MaterialAttributes=@n5.MaterialAttributes, BaseColor=@n51.Combined_Output, Metallic=@n52, Roughness=@n53, Normal=@n5.Normal) {attributeSetTypes [
    #1
    #3
    #4
    #2
   ]}
  n55 = MaterialFunctionCall(
   Gear_mask_Red=@n32
   Gears=@n7
   Normals=@n32
   Uvs_GearA_Mask=@n34.UVs
   Uvs_GearA=@n38.Rotated Values
   Uvs_GearB_Mask=@n40.UVs
   Uvs_GearB=@n44.Rotated Values
   Uvs_GearC_Mask=@n27.UVs
   Uvs_GearC=@n31.Rotated Values
   Uvs_GearD_Mask=@n46.UVs
   Uvs_GearD=@n50.Rotated Values
  ) {materialFunction=&MF_ClockWork_Gears}
  n56 = FunctionInput(Preview=VertexInterpolator(VS=PreSkinnedNormal()).PS) {bUsePreviewValueAsDefault=true, inputName=Local Normals, sortPriority=1}
  n57 = Abs(None=ComponentMask(None=@n56) {r=true})
  n58 = ScalarParameter() {group=Wrap_AlDay, parameterName=Ald_TriblendMask_Contrast}
  n59 = MaterialFunctionCall(In=Abs(None=ComponentMask(None=@n56) {b=true}), Contrast=@n58) {materialFunction=&CheapContrast}
  n60 = MaterialFunctionCall(In=Abs(None=ComponentMask(None=@n56) {g=true}), Contrast=@n58) {materialFunction=&CheapContrast}
  n61 = LinearInterpolate(
   A=LinearInterpolate(
    A=MaterialFunctionCall(In=@n57, Contrast=@n58) {materialFunction=&CheapContrast}.Result
    B=LinearInterpolate(A=@n59.Result, B=@n60.Result, Alpha=@n18)
    Alpha=@n10
   )
   B=@n60.Result
   Alpha=@n11
  )
  n62 = TextureObjectParameter() {group=Wrap_AlDay, parameterName=Ald_BaseColor_02, sortPriority=56, texture=&T_BlacksRGB}
  n63 = VectorParameter() {group=Wrap_AlDay, parameterName=Ald_Rotation-Weapon----------------------, sortPriority=150, defaultValue=<rgba 0 0 -0.25 1>}
  n64 = VectorParameter() {group=Wrap_AlDay, parameterName=Ald_Rotation_Vehicle, sortPriority=151, defaultValue=<rgba 0 0 0 1>}
  n65 = LinearInterpolate(A=@n63.RGB, B=@n64.RGB, Alpha=@n10)
  n66 = VectorParameter() {group=Wrap_AlDay, parameterName=Ald_Rotation_Mech, sortPriority=153, defaultValue=<rgba 0.25 0.25 0 1>}
  n67 = Time()
  n68 = ScalarParameter() {group=Wrap_AlDay, parameterName=Ald_Rotation_BkSpeed, sortPriority=150}
  n69 = LinearInterpolate(
   A=LinearInterpolate(
    A=ScalarParameter() {defaultValue=1, group=Wrap_AlDay, parameterName=Ald_scale_Weapon, sortPriority=164}
    B=ScalarParameter() {defaultValue=1, group=Wrap_AlDay, parameterName=Ald_scale_Vehicle, sortPriority=164}
    Alpha=@n10
   )
   B=ScalarParameter() {defaultValue=1, group=Wrap_AlDay, parameterName=Ald_scale_mech, sortPriority=164}
   Alpha=@n11
  )
  n70 = Constant() {r=0.5}
  n71 = Add(A=Divide(A=@n13.Result, B=@n69), B=@n70)
  n72 = Divide(A=@n70, B=@n69) {desc=Scale the center offset by the scaling factor}
  n73 = VectorParameter() {group=Wrap_AlDay, parameterName=Ald_Offset_Weapon_---------------------, sortPriority=160, defaultValue=<rgba 0 0 0 1>}
  n74 = LinearInterpolate(
   A=@n73.RGB
   B=VectorParameter() {group=Wrap_AlDay, parameterName=Ald_Offset_Vehicle, sortPriority=161, defaultValue=<rgba 0 0 0 1>}.RGB
   Alpha=@n10
  )
  n75 = VectorParameter() {group=Wrap_AlDay, parameterName=Ald_Offset_Mech, sortPriority=162, defaultValue=<rgba 0 0 0 1>}
  n76 = VectorParameter() {group=Wrap_AlDay, parameterName="Ald_Fill_alternate_Axis_Color_A=Roughness (V4)"}
  n77 = MaterialFunctionCall!(
   IsCharacter
   IsStatsB
   Texture=@n62
   Rotation_Axis=Add(A=LinearInterpolate(A=@n65, B=@n66.RGB, Alpha=@n11), B=Multiply(A=@n67, B=@n68))
   Caculated_Bounds=Add(A=Subtract(A=@n71, B=@n72) {desc=Re offset to maintain center}, B=LinearInterpolate(A=@n74, B=@n75.RGB, Alpha=@n11))
   Pre_Skinned_Normals=@n56
   Blend_Contrast=ScalarParameter() {group=Wrap_AlDay, parameterName=Ald_TriblendMask_Contrast}
   Fill_alternate_Axis_Opacity=ScalarParameter() {group=Wrap_AlDay, parameterName="Ald_Fill_alternate_Axis_Opacity (S)"}
   "Fill_alternate_Axis_Color_A=Roughness"=AppendVector(A=@n76.RGB, B=@n76.RGB)
   Is_Vehicle=@n10
   Is_Mech=@n11
   Enable_Fill_Z_Weapon
   Fill_Z_Weapon_opacity
   Fill_Z_Weapon_Color
  ) {materialFunction=&/MF_TopsSideProjection_Base}
  n78 = SetMaterialAttributes!(MaterialAttributes=@n5.MaterialAttributes, BaseColor=@n77.RGB, Metallic=Constant(), Roughness=@n77.RGB) {attributeSetTypes [#1, #3, #4]}
  n79 = BlendMaterialAttributes(A=@n78, B=@n54, Alpha=Multiply(A=@n61, B=Multiply(A=@n55.Gear_Mask_Depth, B=@n55.Gear_Mask_Depth))) {vertexAttributeBlendType=UseB}
  BlendMaterialAttributes(
   A=SetMaterialAttributes(MaterialAttributes=@n5.MaterialAttributes, BaseColor=@n3.RGB, Metallic=Constant(), Roughness=@n3.RGB) {attributeSetTypes [#1, #3, #4]}
   B=@n79
   Alpha=@n1
  ) {vertexAttributeBlendType=UseB}
  Constant()
  n80 = TextureObjectParameter() {group=Wrap_AlDay, parameterName=BC_Ald_02, sortPriority=-1, texture=&T_BlacksRGB}
  n81 = NamedRerouteUsage()
  n82 = NamedRerouteUsage()
  n83 = ComponentMask(None=VectorParameter() {group=Wrap_AlDay, parameterName=Ald_E_Position, defaultValue=<rgba -0.375 -0.375 0 0>}.RGB) {g=true, r=true}
  n84 = MaterialFunctionCall(
   UVs=Multiply(A=Add(A=@n26, B=@n83)) {constB=4}
   Texture Scale=ScalarParameter() {defaultValue=1.1, group=Wrap_AlDay, parameterName=Ald_E_Scale}
  ) {materialFunction=&ScaleUVsByCenter}
  n85 = ComponentMask(None=VectorParameter() {group=Wrap_AlDay, parameterName=Ald_F_Position, defaultValue=<rgba -0.375 -0.375 0 0>}.RGB) {g=true, r=true}
  n86 = MaterialFunctionCall(
   UVs=Multiply(A=Add(A=@n26, B=@n85)) {constB=4}
   Texture Scale=ScalarParameter() {defaultValue=1.1, group=Wrap_AlDay, parameterName=Ald_F_Scale}
  ) {materialFunction=&ScaleUVsByCenter}
  n87 = MaterialFunctionCall!(
   BackGroundMask=@n6
   Gear_mask_Red=@n32
   Normals=@n8
   Gears=@n80
   Uvs_GearE_Mask=@n84.UVs
   Uvs_GearE=@n81
   Uvs_GearF_Mask=@n86.UVs
   Uvs_GearF=@n82
  ) {materialFunction=&/MF_AlDay_Gears}
  n88 = Constant()
  n89 = SetMaterialAttributes!(MaterialAttributes=@n5.MaterialAttributes, BaseColor=@n87.Combined_Output, Metallic=@n88, Roughness=@n53, Normal=@n5.Normal) {attributeSetTypes [
    #1
    #3
    #4
    #2
   ]}
  n90 = MaterialFunctionCall(Gear_mask_Red=@n32, Normals=@n32, Gears=@n80, Uvs_GearE_Mask=@n84.UVs, Uvs_GearE=@n81, Uvs_GearF_Mask=@n86.UVs, Uvs_GearF=@n82) {materialFunction=&MF_AlDay_Gears}
  n91 = Multiply(A=@n61, B=Multiply(A=@n90.Gear_Mask_Depth, B=@n90.Gear_Mask_Depth))
  n92 = GetMaterialAttributes(None=BlendMaterialAttributes(A=BlendMaterialAttributes(A=@n78, B=@n89, Alpha=@n91), B=@n79, Alpha=ComponentMask(None=@n90.Texture_Non_alpha) {g=true}) {vertexAttributeBlendType=UseB}) {attributeGetTypes [#1]}
  n93 = Multiply(A=Time(), B=ScalarParameter() {group=Wrap_AlDay, parameterName=Ald_HueShift_Time})
  n94 = ScalarParameter() {group=Wrap_AlDay, parameterName=Ald_HueShift}
  n95 = SetMaterialAttributes!(
   MaterialAttributes=@n92.MaterialAttributes
   BaseColor=MaterialFunctionCall(Hue Shift Percentage=Add(A=@n93, B=@n94), Texture=@n92.BaseColor) {materialFunction=&HueShift}.Result
  ) {attributeSetTypes [#1]}
  n96 = LinearInterpolate(A=@n1, B=@n2, Alpha=NamedRerouteUsage())
  FunctionOutput(None=StaticSwitchParameter(True=BlendMaterialAttributes(A=@n4, B=@n95, Alpha=@n96) {vertexAttributeBlendType=UseB}, False=@n4) {group=Customization_Wrap, parameterName=Wrap_AlDay})
  Multiply(A=@n87.Gear_Mask, B=@n88)
  Multiply(A=@n51.Gear_Mask, B=@n52)
  NamedRerouteDeclaration(None=FunctionInput() {bUsePreviewValueAsDefault=true, inputName=isCharacter, inputType=FunctionInput_Scalar, sortPriority=4}) {name=IsCharacter, nodeColor=<rgba 0.7812 1 0 1>}
  n97 = Multiply(A=Frac(None=@n84.UVs)) {constB=0.5}
  n98 = Constant2Vector()
  n99 = ComponentMask(None=Constant3Vector() {constant=<rgba 0.25 0.25 0 1>}) {g=true, r=true}
  n100 = Time()
  n101 = ScalarParameter() {defaultValue=-0.12, group=Wrap_AlDay, parameterName=Ald_E_Rotation_Speed}
  NamedRerouteDeclaration(None=MaterialFunctionCall(UVs=Add(A=@n97, B=@n98), Rotation Center=Add(A=@n98, B=@n99), "Rotation Angle (0-1)"=Multiply(A=@n100, B=@n101)) {materialFunction=&CustomRotator}.Rotated Values) {name=IsAld_E_Pos, nodeColor=<rgba 0.2656 0 1 1>}
  n102 = Multiply(A=Frac(None=@n86.UVs)) {constB=0.5}
  n103 = Constant2Vector() {g=0.5}
  n104 = ComponentMask(None=Constant3Vector() {constant=<rgba 0.25 0.25 0 1>}) {g=true, r=true}
  n105 = Time()
  n106 = ScalarParameter() {defaultValue=-0.12, group=Wrap_AlDay, parameterName=Ald_F_Rotation_Speed}
  NamedRerouteDeclaration(None=MaterialFunctionCall(UVs=Add(A=@n102, B=@n103), Rotation Center=Add(A=@n103, B=@n104), "Rotation Angle (0-1)"=Multiply(A=@n105, B=@n106)) {materialFunction=&CustomRotator}.Rotated Values) {name=IsAld_F_Pos, nodeColor=<rgba 0.0781 0 1 1>}
 ]}
