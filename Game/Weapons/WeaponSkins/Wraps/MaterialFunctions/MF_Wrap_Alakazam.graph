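# Alakazam wrap material function that layers a skin and reactive emissive effect onto an incoming material, gated by a static switch.
# isState selects weapon, vehicle or mech settings for side-projected UVs, with normal-based fill on other axes.
# Two customization masks blend in textured, pulsing-glow and flat secondary layers.
# A "ShieldWeaver" overlay animates randomized, staggered flipbook cells through a warped scrolling gradient, fading after a Blueprint-supplied trigger time.
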MaterialFunction /Game/Weapons/WeaponSkins/Wraps/MaterialFunctions/MF_Wrap_Alakazam
graph {nodes [
  n1 = TextureObjectParameter() {
   group=Wrap__ALC_Textures
   parameterName=ALC_Normal_01
   samplerType=SAMPLERTYPE_Normal
   sortPriority=16
   texture=&FlatNormalBlue
  }
  n2 = FunctionInput(Preview=Constant()) {inputName=isState, inputType=FunctionInput_Vector4, sortPriority=5}
  n3 = MaterialFunctionCall(Color=@n2) {materialFunction=&SplitComponents}
  n4 = VectorParameter() {group=Wrap_ALC_UVS, parameterName=ALC_Offset_Weapon, defaultValue=<rgba 0 0 0 1>}
  n5 = MaterialFunctionCall(Color=@n2) {materialFunction=&SplitComponents}
  n6 = LinearInterpolate(
   A=VectorParameter() {group=Wrap_ALC_UVS, parameterName=ALC_Offset_Vehicle, defaultValue=<rgba 0 0 0 1>}.RGB
   B=VectorParameter() {group=Wrap_ALC_UVS, parameterName=ALC_Offset_Vehicle_Side, defaultValue=<rgba 0 0 0 1>}.RGB
   Alpha=@n5.G
  )
  n7 = ScalarParameter() {defaultValue=1, group=Wrap_ALC_UVS, parameterName=ALC_TileRate_Weapon}
  n8 = LinearInterpolate(
   A=ScalarParameter() {defaultValue=1, group=Wrap_ALC_UVS, parameterName=ALC_TileRate_Vehicle}
   B=ScalarParameter() {defaultValue=1, group=Wrap_ALC_UVS, parameterName=ALC_TileRate_VehicleSide}
   Alpha=@n5.G
  )
  n9 = FunctionInput(Preview=VertexInterpolator(VS=PreSkinnedNormal()).PS) {bUsePreviewValueAsDefault=true, inputName=Pre-Skinned Local Normal, sortPriority=10}
  n10 = PreSkinnedLocalBounds()
  n11 = Subtract(A=VertexInterpolator(VS=PreSkinnedPosition()).PS, B=@n10.Min)
  n12 = Subtract(A=@n10.Max, B=@n10.Max)
  n13 = MaterialFunctionCall!(
   For_Safe_Zones
   Normal_Map=@n1
   Caculated_Bounds=FunctionInput(Preview=Divide(A=@n11, B=@n12)) {bUsePreviewValueAsDefault=true, inputName=Calcated_Bounds, sortPriority=11}
   Pre_Skinned_Normals=@n9
   Is Vehicle=@n3.R
   Is VehicleSide=@n3.R
   Is Mech=@n3.R
   Vehicle_Rotation=ScalarParameter() {group=Wrap_ALC_UVS, parameterName=ALC_Rotation_Vehicle}
   UV_Rotation Bypass
   Mech_Rotation=ScalarParameter() {defaultValue=0.25, group=Wrap_ALC_UVS, parameterName=ALC_Rotation_Mech -----}
   Uvs_Offset=LinearInterpolate(
    A=LinearInterpolate(A=@n4.RGB, B=@n6, Alpha=@n5.R)
    B=VectorParameter() {group=Wrap_ALC_UVS, parameterName=ALC_Offset_Mech -----, defaultValue=<rgba 0 0 0 1>}.RGB
    Alpha=@n5.B
   )
   Offset-after rotation
   Uvs_TileRate=LinearInterpolate(
    A=LinearInterpolate(A=@n7, B=@n8, Alpha=@n5.R)
    B=ScalarParameter() {defaultValue=1, group=Wrap_ALC_UVS, parameterName=ALC_TileRate_Mech -----}
    Alpha=@n5.B
   )
  ) {materialFunction=&/MF_Projection_Side_Normal}
  n14 = Constant()
  n15 = VectorParameter() {group=Wrap_ALC_Emissive, parameterName=ALC_ShieldWeaver_Offset_Vehicle_Side, defaultValue=<rgba 0 0 0 1>}
  n16 = MaterialFunctionCall(Float4=@n2) {materialFunction=&BreakOutFloat4Components}
  n17 = LinearInterpolate(
   A=VectorParameter() {group=Wrap_ALC_Emissive, parameterName=ALC_ShieldWeaver_Offset_Weapon, defaultValue=<rgba 0 0 0 1>}.RGB
   B=LinearInterpolate(
    A=VectorParameter() {group=Wrap_ALC_Emissive, parameterName=ALC_ShieldWeaver_Offset_Vehicle, defaultValue=<rgba 0 0 0 1>}.RGB
    B=@n15.RGB
    Alpha=@n16.G
   )
   Alpha=@n16.R
  )
  n18 = VectorParameter() {group=Wrap_ALC_Emissive, parameterName=ALC_ShieldWeaver_Offset_Mech -----, defaultValue=<rgba 0 0 0 1>}
  n19 = Add(A=@n13.UVs No Offsets/Tiling, B=ComponentMask(None=LinearInterpolate(A=@n17, B=@n18.RGB, Alpha=@n16.B)) {g=true, r=true})
  n20 = ScalarParameter() {defaultValue=10, group=Wrap_ALC_Emissive, parameterName=ALC_ShieldWeaver_TileRate_Vehicle, sortPriority=100}
  n21 = ScalarParameter() {defaultValue=10, group=Wrap_ALC_Emissive, parameterName=ALC_ShieldWeaver_TileRate_VehicleSide, sortPriority=100}
  n22 = LinearInterpolate(
   A=LinearInterpolate(
    A=ScalarParameter() {defaultValue=10, group=Wrap_ALC_Emissive, parameterName=ALC_ShieldWeaver_TileRate_Weapon, sortPriority=100}
    B=LinearInterpolate(A=@n20, B=@n21, Alpha=@n16.G)
    Alpha=@n16.R
   )
   B=LinearInterpolate(
    A=ScalarParameter() {defaultValue=12, group=Wrap_ALC_Emissive, parameterName=ALC_ShieldWeaver_TileRate_Mech-----, sortPriority=100}
    B=ScalarParameter() {defaultValue=12, group=Wrap_ALC_Emissive, parameterName=ALC_ShieldWeaver_TileRate_Character, sortPriority=100}
    Alpha=@n16.A
   )
   Alpha=@n16.B
  )
  n23 = Multiply(A=ComponentMask(None=@n19) {r=true}, B=@n22) {constB=5}
  n24 = Multiply(
   A=Divide(A=Floor(None=@n23), B=@n22) {constB=5}
   B=ScalarParameter() {defaultValue=0.5, group=Wrap_ALC_Emissive, parameterName=ALC_ShieldWeaver_Offset_Y-----, sortPriority=40}
  )
  n25 = Multiply(A=Add(A=@n19, B=AppendVector(A=@n14, B=@n24)) {constB=0}, B=@n22)
  n26 = VectorParameter() {group=Wrap_ALC_Emissive, parameterName=ALC_ShieldWeaver_Random_seed, sortPriority=40, defaultValue=<rgba 4 3 0 0>}
  n27 = MaterialFunctionCall!(2 Vector=Add(A=Floor(None=@n25), B=ComponentMask(None=@n26.RGB) {g=true, r=true}), RandomSeed) {materialFunction=&/MF_Random_2}
  n28 = LinearInterpolate(Alpha=@n27.utput)
  n29 = ScalarParameter() {defaultValue=0.25, group=Wrap_ALC_Emissive, parameterName=ALC_ShieldWeaver_WarpAmount}
  n30 = VectorParameter() {group=Wrap_ALC_Emissive, parameterName=ALC_ShieldWeaver_GradientSpeed, defaultValue=<rgba 0.75 0.75 0 1>}
  n31 = ComponentMask(None=@n30.RGB) {g=true, r=true}
  n32 = Time()
  n33 = Add(A=@n13.UVs No Offsets/Tiling, B=Add(A=Multiply(A=@n28, B=@n29) {constB=0.5}, B=Multiply(A=@n31, B=@n32)))
  n34 = VectorParameter() {group=Wrap_ALC_Reactive, parameterName=ALC_Reactive_TopLft_Offset, defaultValue=<rgba 0 0 0 1>}
  n35 = VectorParameter() {group=Wrap_ALC_Reactive, parameterName=ALC_Reactive_TopLft_Offset_End, defaultValue=<rgba 0 0 0 1>}
  n36 = Time()
  n37 = ScalarParameter() {defaultValue=-9999999, group=Wrap_ALC_Reactive, parameterName=TimeFrom-BP----------------, sortPriority=-600}
  n38 = Saturate(None=Divide(
    A=Subtract(A=@n36, B=@n37)
    B=ScalarParameter() {defaultValue=0.25, group=Wrap_ALC_Reactive, parameterName=ALC_AnimationDuration, sortPriority=33}
   ))
  Add(A=@n33, B=ComponentMask(None=LinearInterpolate(A=@n34.RGB, B=@n35.RGB, Alpha=@n38)) {g=true, r=true})
  n39 = TextureObjectParameter() {group=Wrap__ALC_Textures, parameterName=ALC_Mask_02, sortPriority=20, texture=&T_BlacksRGB}
  n40 = TextureSample(UVs=@n13.UVs, Tex=@n39) {texture=&T_BlacksRGB}
  n41 = ChannelMaskParameter(None=@n40.RGBA) {group=Wrap_ALC, maskChannel=Alpha, parameterName=ALC_MetalChannel_Mask_01, sortPriority=-26, defaultValue=<rgba 0 0 0 1>}
  n42 = MaterialFunctionCall(Color=Abs(None=@n9)) {materialFunction=&SplitComponents}
  n43 = MaterialFunctionCall(Color=@n2) {materialFunction=&SplitComponents}
  n44 = LinearInterpolate(A=Add(A=@n42.R, B=@n42.R), B=Add(A=@n42.B, B=@n42.B), Alpha=@n43.R)
  n45 = Add(A=@n42.R, B=@n42.R)
  n46 = ScalarParameter() {defaultValue=0.1, group=Wrap_ALC, parameterName=ALC_Mask_Contrast, sortPriority=51}
  n47 = OneMinus(None=MaterialFunctionCall(In=LinearInterpolate(A=@n44, B=@n45, Alpha=@n43.B), Contrast=@n46) {materialFunction=&CheapContrast}.Result)
  n48 = MaterialFunctionCall(Color=Abs(None=@n9)) {materialFunction=&SplitComponents}
  n49 = LinearInterpolate(A=@n48.R, B=LinearInterpolate(A=@n48.B, B=@n48.B, Alpha=@n43.G), Alpha=@n43.R)
  n50 = OneMinus(None=MaterialFunctionCall(In=LinearInterpolate(A=@n49, B=@n48.G, Alpha=@n43.B), Contrast=@n46) {materialFunction=&CheapContrast}.Result)
  n51 = ScalarParameter() {defaultValue=1, group=Wrap_ALC, parameterName=ALC_Mask_Z+AltAxisActive -----, sortPriority=50}
  n52 = LinearInterpolate(
   A=LinearInterpolate(
    A=Constant()
    B=Saturate(None=@n47)
    Alpha=ScalarParameter() {defaultValue=1, group=Wrap_ALC, parameterName=ALC_Mask_AltAxisActive -----, sortPriority=50}
   )
   B=LinearInterpolate(A=Constant(), B=Saturate(None=@n50), Alpha=@n51)
   Alpha=@n51
  )
  n53 = ScalarParameter() {group=Wrap_ALC, parameterName=ALC_Specular_Min, sortPriority=1}
  n54 = ScalarParameter() {defaultValue=1, group=Wrap_ALC, parameterName=ALC_Specular_Max, sortPriority=1}
  n55 = ChannelMaskParameter(None=@n40.RGBA) {group=Wrap_ALC, parameterName=ALC_SpecChannel_Mask_01, sortPriority=-26}
  n56 = ScalarParameter() {group=Wrap_ALC, parameterName=ALC_Roughness, sortPriority=1}
  n57 = ScalarParameter() {defaultValue=1, group=Wrap_ALC, parameterName=ALC_Roughness_High, sortPriority=1}
  n58 = TextureObjectParameter() {group=Wrap__ALC_Textures, parameterName=ALC_BaseColor_01, sortPriority=11, texture=&T_BlacksRGB}
  n59 = TextureSample(UVs=@n13.UVs, Tex=@n58) {texture=&T_BlacksRGB}
  n60 = ChannelMaskParameter(None=@n59.RGBA) {group=Wrap_ALC, maskChannel=Alpha, parameterName=ALC_RoughChannel_Mask_01, sortPriority=-26, defaultValue=<rgba 0 0 0 1>}
  n61 = VectorParameter() {group=Wrap_ALC, parameterName="ALC_Fill_alternate_Axis_Color_A=Roughness 01", defaultValue=<rgba 0.4735 0.4508 0.4179 0.5>}
  n62 = LinearInterpolate(A=@n13.Raw_Normal, B=Constant3Vector(), Alpha=@n52)
  n63 = GetMaterialAttributes(None=FunctionInput() {inputName=Material, inputType=FunctionInput_MaterialAttributes, sortPriority=-20}) {attributeGetTypes [#1]}
  n64 = VectorParameter() {group=Wrap_ALC_Emissive, parameterName=ALC_Emissive_Color-----, defaultValue=<rgba 0.0018 0.6105 0.9047 1>}
  n65 = Time()
  n66 = ScalarParameter() {defaultValue=0.1, group=Wrap_ALC_Emissive, parameterName=ALC_Emissive_PulseTime}
  n67 = Add(A=Sine(None=Multiply(A=@n65, B=@n66) {constB=0.1}) {period=0.5})
  n68 = Multiply(
   A=Clamp(
    None=Divide(A=@n67)
    Min=ScalarParameter() {defaultValue=0.1, group=Wrap_ALC_Emissive, parameterName=ALC_Emissive_PulseMin}
    Max=ScalarParameter() {defaultValue=1, group=Wrap_ALC_Emissive, parameterName=ALC_Emissive_PulseMax}
   )
   B=ScalarParameter() {defaultValue=6, group=Wrap_ALC_Emissive, parameterName=ALC_Emissive_Amount}
  )
  n69 = TextureObjectParameter() {group=Wrap__ALC_Textures, parameterName=ALC_Mask_03, sortPriority=20, texture=&T_BlacksRGB}
  n70 = ChannelMaskParameter(None=TextureSample(UVs=@n13.UVs, Tex=@n69) {texture=&T_BlacksRGB}.RGBA) {group=Wrap_ALC_Emissive, parameterName=ALC_Emissive_GlowMask_01}
  n71 = Multiply(A=Time(), B=ScalarParameter() {defaultValue=0.1, group=Wrap_ALC_Emissive, parameterName=ALC_Emissive_BasePulseTime}) {constB=0.1}
  n72 = Divide(A=Add(A=Sine(None=@n71) {period=0.5}))
  n73 = TextureObjectParameter() {group=Wrap__ALC_Textures, parameterName=ALC_Mask_01, sortPriority=20, texture=&T_BlacksRGB}
  n74 = Add(
   A=Multiply(A=@n64.RGB, B=Multiply(A=@n68, B=@n70.R))
   B=Multiply(A=Clamp(None=@n72) {minDefault=0.35}, B=TextureSample(UVs=@n13.UVs, Tex=@n73) {texture=&T_BlacksRGB}.RGB)
  )
  n75 = ScalarParameter() {defaultValue=4, group=Wrap_ALC_Emissive, parameterName=ALC_ShieldWeaver_EmissiveAmount}
  n76 = ScalarParameter() {group=Wrap_ALC_Reactive, parameterName=ALC_TestReactive, sortPriority=300}
  n77 = LinearInterpolate(A=OneMinus(None=@n38) {desc=inverts curve from 0-1 to 1-0}, B=@n76, Alpha=@n76)
  n78 = SetMaterialAttributes!(
   MaterialAttributes=@n63.MaterialAttributes
   BaseColor=LinearInterpolate(A=ComponentMask(None=@n59.RGBA) {b=true, g=true, r=true}, B=@n61.RGB, Alpha=@n52)
   Metallic=LinearInterpolate(A=@n41.R, B=ScalarParameter() {group=Wrap_ALC, parameterName=ALC_Metal_Mask, sortPriority=1}, Alpha=@n52)
   Specular=LinearInterpolate(
    A=LinearInterpolate(A=@n53, B=@n54, Alpha=@n55.R)
    B=ScalarParameter() {defaultValue=0.5, group=Wrap_ALC, parameterName=ALC_Specular_Mask, sortPriority=1}
    Alpha=@n52
   )
   Roughness=LinearInterpolate(A=LinearInterpolate(A=@n56, B=@n57, Alpha=@n60.R), B=@n61.A, Alpha=@n52)
   Normal=LinearInterpolate(
    A=@n63.Normal
    B=Add(A=@n63.Normal, B=@n62)
    Alpha=ScalarParameter() {defaultValue=1, group=Wrap_ALC, parameterName=ALC_Normal_Intensty, sortPriority=1}
   )
   EmissiveColor=LinearInterpolate(A=@n74, B=Multiply(A=@n74, B=@n75), Alpha=@n77)
  ) {attributeSetTypes [
    #2
    #3
    #4
    #5
    #1
    #6
   ]}
  n79 = FunctionInput() {
   bUsePreviewValueAsDefault=true
   desc=secondary
   inputName="Customization Mask 2 (_CM B Top Right)"
   inputType=FunctionInput_Scalar
   sortPriority=50
  }
  n80 = BlendMaterialAttributes(A=@n63.MaterialAttributes, B=@n78, Alpha=@n79) {vertexAttributeBlendType=UseA}
  n81 = VectorParameter() {group=Wrap_ALC_Secondary, parameterName=ALC_Secondary_BaseColor, defaultValue=<rgba 0.0762 0.0887 0.1221 1>}
  n82 = SetMaterialAttributes!(
   MaterialAttributes=GetMaterialAttributes(None=@n80).MaterialAttributes
   BaseColor=@n81.RGB
   Metallic=ScalarParameter() {group=Wrap_ALC_Secondary, parameterName=ALC_Secondary_Metallic}
   Specular=ScalarParameter() {defaultValue=0.55, group=Wrap_ALC_Secondary, parameterName=ALC_Secondary_Specular}
   Roughness=ScalarParameter() {defaultValue=0.45, group=Wrap_ALC_Secondary, parameterName=ALC_Secondary_Roughness}
   Normal=@n63.Normal
  ) {attributeSetTypes [
    #2
    #3
    #4
    #5
    #1
   ]}
  n83 = FunctionInput() {
   bUsePreviewValueAsDefault=true
   desc=base customization
   inputName="Customization Mask (_CM B Top left)"
   inputType=FunctionInput_Scalar
   sortPriority=49
  }
  n84 = BlendMaterialAttributes(A=@n80, B=@n82, Alpha=@n83) {vertexAttributeBlendType=UseA}
  n85 = GetMaterialAttributes(None=@n84) {attributeGetTypes [#2, #6]}
  n86 = VectorParameter() {group=Wrap_ALC_Emissive, parameterName=ALC_ShieldWeaver_BaseColor-----, defaultValue=<rgba 0.0507 0 0.622 1>}
  n87 = ChannelMaskParameter(None=TextureSample(UVs=@n33, Tex=@n69) {texture=&T_BlacksRGB}.RGBA) {group=Wrap_ALC_Emissive, maskChannel=Green, parameterName=ALC_ShieldWeaver_Gradient, defaultValue=<rgba 0 1 0 0>}
  n88 = Constant() {r=4}
  n89 = MaterialFunctionCall(
   UVs=Frac(None=@n25)
   Texture Scale=ComponentMask(None=VectorParameter() {group=Wrap_ALC_Emissive, parameterName=ALC_ShieldWeaver_Scale, defaultValue=<rgba 1.035 0.975 0 1>}.RGB) {g=true, r=true}
  ) {materialFunction=&ScaleUVsByCenter}
  n90 = MaterialFunctionCall!(UVs=@n19, Texture Size=Constant() {r=4}) {materialFunction=/Engine/Functions/Engine_MaterialFunctions01/Texturing/ComputeMipLevel}
  n91 = StaticBool() {value=true}
  n92 = Multiply(
   A=OneMinus(None=@n52)
   B=Multiply(
    A=MaterialFunctionCall(
     "Animation  Phase (0-1)"=@n87.R
     Number of Rows=@n88
     Number of Columns=@n88
     Texture=@n73
     UVs=@n89.UVs
     MipBias/Level=@n90.MipLevel
     "Use Mip Bias (T) Level (F)"=@n91
    ) {materialFunction=&FlipBook}.Alpha
    B=Round(None=@n27.utput)
   )
  )
  n93 = Power(Base=@n92) {constExponent=0.5}
  n94 = Power(Base=@n93) {constExponent=0.5}
  n95 = VectorParameter() {group=Wrap_ALC_Emissive, parameterName=ALC_ShieldWeaver_Color-----, defaultValue=<rgba 0.4397 0.9216 0.9823 1>}
  n96 = Multiply(A=Multiply(A=@n95.RGB, B=@n75), B=@n92)
  n97 = SetMaterialAttributes!(
   MaterialAttributes=@n85.MaterialAttributes
   BaseColor=LinearInterpolate(A=@n85.BaseColor, B=LinearInterpolate(A=@n85.BaseColor, B=@n86.RGB, Alpha=@n94), Alpha=@n77)
   EmissiveColor=Add(A=@n85.EmissiveColor, B=LinearInterpolate(B=@n96, Alpha=@n77))
  ) {attributeSetTypes [#2, #6]}
  n98 = Constant() {r=1}
  FunctionOutput(None=StaticSwitchParameter(True=BlendMaterialAttributes(A=@n84, B=@n97, Alpha=@n98) {vertexAttributeBlendType=UseA}, False=@n63.MaterialAttributes) {group=Customization_Wrap, parameterName=Wrap_Alakazam})
  Multiply(A=@n85.BaseColor)
  n99 = OneMinus(None=@n93)
  Multiply(B=MaterialFunctionCall(
    In=LinearInterpolate(B=@n99, Alpha=@n77) {constA=1}
    Contrast=ScalarParameter() {defaultValue=2, group=Wrap_ALC_Emissive, parameterName=ALC_DarkenBase_Contrast, sortPriority=0}
   ) {materialFunction=&CheapContrast}.Result)
  Multiply(A=@n52, B=@n13.Raw_Normal)
  NamedRerouteDeclaration(None=MaterialFunctionCall(Float4=@n2) {materialFunction=&BreakOutFloat4Components}.A) {name=IsCharacter, nodeColor=<rgba 1 0.9609 0 1>}
 ]}
