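# Amber wrap material function that layers an animated amber look with an embedded bug and bubbles over weapons, vehicles and mechs.
# Bounds-based triplanar projection supplies surface mask and normals, with scale and rotation chosen by the IsState input.
# Camera-offset bug and bubble UVs fake depth, combining with marbling, fake light and a camera-facing inner sphere mask.
# Curve-atlas lookups color the time-of-day-scaled emissive, blended through the customization mask behind the Wrap_Amber static switch.
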MaterialFunction /Game/Weapons/WeaponSkins/Wraps/MaterialFunctions/MF_Wrap_Amber
graph {nodes [
  n1 = TextureObjectParameter() {
   group=Wrap_Amber
   parameterName=Amber_BS_Mask
   samplerType=SAMPLERTYPE_Masks
   texture=/Game/Weapons/WeaponSkins/Wraps/Textures/Amber/T_Wrap_Amber_M_1
  }
  n2 = ScalarParameter() {defaultValue=1, group=Wrap_Amber, parameterName=Amber_Tri_Planer_Mask_Contrast, sortPriority=0}
  n3 = ScalarParameter() {group=Wrap_Amber, parameterName=Amber_BS_Fill_alternate_Axis_Opacity}
  n4 = VectorParameter() {group=Wrap_Amber, parameterName="Amber_BS_alternate_Axis_Color_A=Roughness ", defaultValue=<rgba 0 0 0 1>}
  n5 = FunctionInput(Preview=Constant()) {bUsePreviewValueAsDefault=true, inputName=IsState, inputType=FunctionInput_Vector4}
  n6 = MaterialFunctionCall(Color=@n5) {materialFunction=&SplitComponents}
  n7 = FunctionInput(Preview=VertexInterpolator(VS=PreSkinnedNormal()).PS) {bUsePreviewValueAsDefault=true, inputName=Pre-Skinned Local Normal, sortPriority=10}
  n8 = MaterialFunctionCall(Color=@n5) {materialFunction=&SplitComponents}
  n9 = LinearInterpolate(
   A=ScalarParameter() {defaultValue=1, group=Wrap_Amber_UVS, parameterName=Amber_BS_Scale_Weapon}
   B=LinearInterpolate(
    A=ScalarParameter() {defaultValue=1, group=Wrap_Amber_UVS, parameterName=Amber_BS_Scale_Vehicle}
    B=ScalarParameter() {defaultValue=1, group=Wrap_Amber_UVS, parameterName=Amber_BS_Scale_Vehicle_Side}
    Alpha=@n8.G
   )
   Alpha=@n8.R
  )
  n10 = ScalarParameter() {defaultValue=1, group=Wrap_Amber_UVS, parameterName=Amber_BS_Scale_Mech}
  n11 = PreSkinnedPosition()
  n12 = PreSkinnedLocalBounds()
  n13 = Divide(A=Subtract(A=VertexInterpolator(VS=@n11).PS, B=@n12.Min), B=Subtract(A=@n12.Max, B=@n12.Max))
  n14 = FunctionInput(Preview=@n13) {bUsePreviewValueAsDefault=true, inputName=Calcated_Bounds, sortPriority=11}
  n15 = Add(
   A=Multiply(A=@n14, B=LinearInterpolate(A=@n9, B=@n10, Alpha=@n8.B))
   B=Multiply(A=Time(), B=VectorParameter() {group=Wrap_Amber, parameterName=Amber_surface_Speed_A, defaultValue=<rgba 0 0 0 1>}.RGB)
  )
  n16 = VectorParameter() {group=Wrap_Amber_UVS, parameterName=Amber_BS-Rotation-Weapon-01---------------------, sortPriority=1, defaultValue=<rgba 0 0 -0.25 1>}
  n17 = VectorParameter() {group=Wrap_Amber_UVS, parameterName=Amber_BS_Rotation_Vehicle_01, sortPriority=2, defaultValue=<rgba 0 0 0 1>}
  n18 = VectorParameter() {group=Wrap_Amber_UVS, parameterName=Amber_BS_Rotation_Vehicle_Side_01, sortPriority=2, defaultValue=<rgba 0 0 0 1>}
  n19 = VectorParameter() {group=Wrap_Amber_UVS, parameterName=Amber_BS_Rotation_Mech_01, sortPriority=3, defaultValue=<rgba 0.25 0.25 0 1>}
  n20 = LinearInterpolate(A=LinearInterpolate(A=@n16.RGB, B=LinearInterpolate(A=@n17.RGB, B=@n18.RGB, Alpha=@n8.G), Alpha=@n8.R), B=@n19.RGB, Alpha=@n8.B)
  n21 = MaterialFunctionCall!(
   IsCharacter
   IsStatsB
   Texture=@n1
   Rotation_Axis=@n20
   Caculated_Bounds=@n15
   Pre_Skinned_Normals=@n7
   Blend_Contrast=@n2
   Fill_alternate_Axis_Opacity=@n3
   "Fill_alternate_Axis_Color_A=Roughness"=AppendVector(A=@n4.RGB, B=@n4.RGB)
   Is_Vehicle=@n6.R
   Is_Mech=@n6.R
   Enable_Fill_Z_Weapon
   Fill_Z_Weapon_opacity
   Fill_Z_Weapon_Color
  ) {materialFunction=&/MF_TopsSideProjection_Base}
  n22 = ComponentMask(None=@n21.Combined) {a=true}
  n23 = PreSkinnedLocalBounds()
  n24 = Add(
   A=@n23.Min
   B=FunctionInput(Preview=Constant3Vector() {constant=<rgba 0 5 0 0>}) {bUsePreviewValueAsDefault=true, inputName=Object Bounds Min Adjustment, sortPriority=50}
  ) {constB=0}
  n25 = FunctionInput(Preview=Constant3Vector() {constant=<rgba -5 -25 -5 0>}) {bUsePreviewValueAsDefault=true, inputName=Object Bounds Max Adjustment, sortPriority=50}
  n26 = Multiply(A=Subtract(A=Add(A=@n23.Max, B=@n25) {constB=0}, B=@n24)) {constB=0.5}
  n27 = MaterialFunctionCall(Float4=@n5) {materialFunction=&BreakOutFloat4Components}
  n28 = LinearInterpolate(A=TransformPosition(None=Add(A=@n24, B=@n26)) {transformType=TRANSFORMPOSSOURCE_World}, B=ObjectPositionWS(), Alpha=@n27.B)
  n29 = Normalize(VectorInput=Subtract(A=@n28, B=CameraPositionWS()))
  n30 = Normalize(VectorInput=CrossProduct(A=@n29, B=Constant3Vector() {constant=<rgba 0 0 1 1>}))
  n31 = MaterialFunctionCall(Float3=@n23.Extents) {materialFunction=&BreakOutFloat3Components}
  n32 = Multiply(
   A=MaterialFunctionCall() {materialFunction=/Engine/Functions/Engine_MaterialFunctions02/WorldPositionOffset/ObjectScale}.Scale X
   B=Max(A=Max(A=@n31.R, B=@n31.R), B=@n31.B)
  )
  n33 = NamedRerouteUsage()
  n34 = NamedRerouteUsage()
  n35 = Add(
   A=LinearInterpolate(
    A=MaterialFunctionCall!(Plane Normal Axis=@n29, Plane X Axis=@n30, Plane Center=@n28, UV Scale=@n32) {materialFunction=/Game/Athena/Environments/Nexus/Materials/VirtualPlaneCoordinatesAbs}.UVs
    B=NamedRerouteUsage()
    Alpha=Add(A=@n33, B=@n34) {constB=0}
   )
   B=ComponentMask(None=VectorParameter() {group=Wrap_Amber, parameterName=Amber_Inner_Wrap_offset, defaultValue=<rgba 0 0 0 1>}.RGB) {g=true, r=true}
  )
  Add(A=Multiply(A=@n22, B=ScalarParameter() {group=Wrap_Amber, parameterName=Amber_Inner_Warp_BS}), B=@n35)
  n36 = Transform(None=CameraVectorWS()) {transformSourceType=TRANSFORMSOURCE_World, transformType=TRANSFORM_Local}
  n37 = ScalarParameter() {group=Wrap_Amber, parameterName=Amber_Bug_Camera}
  n38 = MaterialFunctionCall(Color=Add(A=@n14, B=Multiply(A=@n36, B=@n37))) {materialFunction=&SplitComponents}
  n39 = MaterialFunctionCall(Color=@n5) {materialFunction=&SplitComponents}
  n40 = LinearInterpolate(A=AppendVector(A=@n38.G, B=OneMinus(None=@n38.B)), B=AppendVector(A=@n38.R, B=@n38.R), Alpha=@n39.R)
  n41 = LinearInterpolate(A=AppendVector(A=@n38.G, B=OneMinus(None=@n38.R)), B=AppendVector(A=@n38.R, B=OneMinus(None=@n38.B)), Alpha=@n39.R)
  n42 = ScalarParameter() {defaultValue=0.5, group=Wrap_Amber, parameterName=Amber_Bug_Skew}
  n43 = LinearInterpolate(
   A=LinearInterpolate(
    A=VectorParameter() {group=Wrap_Amber, parameterName=Amber_Bug_Scale_Weapon, defaultValue=<rgba 1 2 1 1>}.RGB
    B=VectorParameter() {group=Wrap_Amber, parameterName=Amber_Bug_Scale_Vehicle, defaultValue=<rgba 1 2 1 1>}.RGB
    Alpha=@n39.R
   )
   B=VectorParameter() {group=Wrap_Amber, parameterName=Amber_Bug_Scale_Mech, defaultValue=<rgba 1 2 1 1>}.RGB
   Alpha=@n39.B
  )
  n44 = LinearInterpolate(
   A=VectorParameter() {group=Wrap_Amber, parameterName=Amber_Bug_Offset_Weapon, defaultValue=<rgba 0 0.5 0 1>}.RGB
   B=VectorParameter() {group=Wrap_Amber, parameterName=Amber_Bug_Offset_Vehicle, defaultValue=<rgba 0 0.5 0 1>}.RGB
   Alpha=@n39.R
  )
  n45 = VectorParameter() {group=Wrap_Amber, parameterName=Amber_Bug_Offset_Mech, defaultValue=<rgba 0 0.5 0 1>}
  n46 = Add(
   A=Multiply(A=LinearInterpolate(A=@n40, B=@n41, Alpha=@n42), B=ComponentMask(None=@n43) {g=true, r=true})
   B=ComponentMask(None=LinearInterpolate(A=@n44, B=@n45.RGB, Alpha=@n39.B)) {g=true, r=true}
  )
  Add(A=Multiply(A=ScalarParameter() {group=Wrap_Amber, parameterName=Amber_Bug_Warp_BS}, B=@n22), B=@n46)
  n47 = Multiply(A=ScalarParameter() {group=Wrap_Amber, parameterName=Amber_Bubbles_Warp_BS}, B=@n22)
  n48 = MaterialFunctionCall(Color=@n5) {materialFunction=&SplitComponents}
  n49 = ComponentMask(None=LinearInterpolate(
    A=VectorParameter() {group=Wrap_Amber, parameterName=Amber_Bubble_Offset, defaultValue=<rgba 0 0.5 0 1>}.RGB
    B=VectorParameter() {group=Wrap_Amber, parameterName=Amber_Bubble_Offset_Mech, defaultValue=<rgba 0 0.5 0 1>}.RGB
    Alpha=@n48.B
   )) {g=true, r=true}
  n50 = ComponentMask(None=LinearInterpolate(
    A=VectorParameter() {group=Wrap_Amber, parameterName=Amber_Bubble_Scale, defaultValue=<rgba 1 2 0 1>}.RGB
    B=VectorParameter() {group=Wrap_Amber, parameterName=Amber_Bubble_Scale_Mech, defaultValue=<rgba 1 2 0 1>}.RGB
    Alpha=@n48.B
   )) {g=true, r=true}
  n51 = ScalarParameter() {group=Wrap_Amber, parameterName=Amber_BubbleB_Camera}
  n52 = MaterialFunctionCall(Color=Add(A=@n14, B=Multiply(A=@n36, B=@n51))) {materialFunction=&SplitComponents}
  n53 = MaterialFunctionCall(Color=@n5) {materialFunction=&SplitComponents}
  n54 = LinearInterpolate(A=AppendVector(A=@n52.G, B=OneMinus(None=@n52.B)), B=AppendVector(A=@n52.R, B=@n52.R), Alpha=@n53.R)
  n55 = LinearInterpolate(A=AppendVector(A=@n52.G, B=OneMinus(None=@n52.R)), B=AppendVector(A=@n52.R, B=OneMinus(None=@n52.B)), Alpha=@n53.R)
  n56 = ScalarParameter() {defaultValue=0.5, group=Wrap_Amber, parameterName=Amber_Bubbles_Skew A/B}
  n57 = Add(A=@n49, B=Multiply(A=@n50, B=LinearInterpolate(A=@n54, B=@n55, Alpha=@n56)))
  Add(A=@n47, B=@n57)
  n58 = ScalarParameter() {defaultValue=1, group=Wrap_Amber_UVS, parameterName=Amber_Marble_Scale_Vehicle}
  n59 = ScalarParameter() {defaultValue=1, group=Wrap_Amber_UVS, parameterName=Amber_Marble_Scale_Vehicle_Side}
  n60 = MaterialFunctionCall(Color=@n5) {materialFunction=&SplitComponents}
  n61 = LinearInterpolate(
   A=LinearInterpolate(
    A=ScalarParameter() {defaultValue=1, group=Wrap_Amber_UVS, parameterName=Amber_Marble_Scale_Weapon}
    B=LinearInterpolate(A=@n58, B=@n59, Alpha=@n60.G)
    Alpha=@n60.R
   )
   B=ScalarParameter() {defaultValue=1, group=Wrap_Amber_UVS, parameterName=Amber_Marble_Scale_Mech}
   Alpha=@n60.B
  )
  n62 = MaterialFunctionCall(Color=@n14) {materialFunction=&SplitComponents}
  n63 = AppendVector(A=@n62.G, B=OneMinus(None=@n62.B))
  n64 = AppendVector(A=@n62.R, B=@n62.R)
  n65 = MaterialFunctionCall(Color=@n5) {materialFunction=&SplitComponents}
  n66 = AppendVector(A=@n62.G, B=OneMinus(None=@n62.R))
  n67 = AppendVector(A=@n62.R, B=OneMinus(None=@n62.B))
  n68 = Multiply(
   A=@n61
   B=LinearInterpolate(
    A=LinearInterpolate(A=@n63, B=@n64, Alpha=@n65.R)
    B=LinearInterpolate(A=@n66, B=@n67, Alpha=@n65.R)
    Alpha=ScalarParameter() {defaultValue=0.5, group=Wrap_Amber, parameterName=Amber_Marble_Skew}
   )
  )
  Add(A=Multiply(A=@n22, B=ScalarParameter() {group=Wrap_Amber, parameterName=Amber_Marble_Warp_BS}), B=@n68)
  n69 = ScalarParameter() {group=Wrap_Amber, parameterName=Amber_BubbleA_Camera}
  n70 = MaterialFunctionCall(Color=Add(A=@n14, B=Multiply(A=@n36, B=@n69))) {materialFunction=&SplitComponents}
  n71 = LinearInterpolate(A=AppendVector(A=@n70.G, B=OneMinus(None=@n70.B)), B=AppendVector(A=@n70.R, B=@n70.R), Alpha=@n48.R)
  n72 = LinearInterpolate(A=AppendVector(A=@n70.G, B=OneMinus(None=@n70.R)), B=AppendVector(A=@n70.R, B=OneMinus(None=@n70.B)), Alpha=@n48.R)
  n73 = Add(A=Multiply(A=LinearInterpolate(A=@n71, B=@n72, Alpha=@n56), B=@n50), B=@n49)
  Add(A=@n73, B=@n47)
  Constant() {r=1}
  n74 = FunctionInput(Preview=SetMaterialAttributes!(MaterialAttributes, Normal=TextureSample() {samplerType=SAMPLERTYPE_Normal, texture=/Game/Weapons/FORT_Rifles/T_SCAR_N}.RGB) {attributeSetTypes [#1]}) {inputName=Material, inputType=FunctionInput_MaterialAttributes, sortPriority=-20}
  n75 = GetMaterialAttributes(None=@n74)
  n76 = VectorParameter() {group=Wrap_Amber, parameterName=Amber_BaseColor, sortPriority=0, defaultValue=<rgba 0.003 0.003 0.003 1>}
  n77 = VectorParameter() {
   group=Wrap_Amber
   parameterName=Amber_BaseColor_Curavture
   channelNames {a="0= BaseColor to Emissive", b="", g="", r=""}
   defaultValue=<rgba 1 1 1 1>
  }
  n78 = FunctionInput(Preview=TextureSample() {samplerType=SAMPLERTYPE_Masks, texture=/Game/Weapons/FORT_Rifles/Textures/T_Scar_CM}.G) {bUsePreviewValueAsDefault=true, inputName=Curavture, inputType=FunctionInput_Scalar, sortPriority=24}
  n79 = CurveAtlasRowParameter(CurveTime=@n78) {
   atlas=/Game/Weapons/WeaponSkins/Wraps/Textures/CruveAtlas_Wraps_01
   curve=/Game/Weapons/WeaponSkins/Wraps/Textures/Amber/C_Amber_Mask
   defaultValue=30
   group=Wrap_Amber
   parameterName=Amber_Mask_ColorCurve
  }
  n80 = ChannelMaskParameter(None=MaterialFunctionCall!(
    Index=ScalarParameter() {defaultValue=30, group=Wrap_Amber, parameterName="Amber_Mask_ColorCurve  Index"}
    CCinput=@n79
    CCinput_Alpha=@n79
    UV=@n78
   ) {materialFunction=&/MF_Global_CC_Tex}.4v) {group=Wrap_Amber, maskChannel=Green, parameterName=Amber_Mask_ColorCurve Channel, defaultValue=<rgba 0 1 0 0>}
  n81 = Add(
   A=ComponentMask(None=@n21.Combined) {g=true}
   B=Multiply(A=@n80.R, B=ScalarParameter() {group=Wrap_Amber, parameterName=Amber_Roghness_Curvature, sortPriority=0})
  )
  n82 = LinearInterpolate(A=LinearInterpolate(B=@n77, Alpha=@n80.R), Alpha=@n77."0= BaseColor to Emissive") {constB=0}
  n83 = Add(
   A=OneMinus(None=@n78)
   B=Multiply(A=@n22, B=ScalarParameter() {defaultValue=0.25, group=Wrap_Amber, parameterName=Amber_Warp_BS_Amount}) {constB=0.25}
  )
  n84 = Add(A=Multiply(A=@n83, B=ScalarParameter() {group=Wrap_Amber, parameterName=Amber_Marble_Warp_Curvature}), B=@n68)
  n85 = Abs(None=@n7)
  n86 = ComponentMask(None=@n85) {b=true}
  n87 = Add(A=ComponentMask(None=@n85) {r=true}, B=@n86)
  n88 = Add(A=@n86, B=ComponentMask(None=@n85) {g=true})
  n89 = MaterialFunctionCall(
   In=LinearInterpolate(
    A=Saturate(None=@n87)
    B=Saturate(None=@n88)
    Alpha=MaterialFunctionCall(Color=@n5) {materialFunction=&SplitComponents}.R
   )
   Contrast=ScalarParameter() {defaultValue=1, group=Wrap_Amber, parameterName=Amber_Tri_Planer_Mask_Contrast, sortPriority=0}
  ) {materialFunction=&CheapContrast}
  n90 = OneMinus(None=@n89.Result)
  n91 = Multiply(A=ScalarParameter() {group=Wrap_Amber, parameterName=Amber_Bubbles_Warp_Curvature}, B=@n83)
  n92 = Add(A=@n73, B=@n91)
  n93 = VectorParameter() {group=Wrap_Amber, parameterName=Amber_Bubbles_Speed_A, defaultValue=<rgba 0 0 0 1>}
  n94 = Multiply(A=Time(), B=AppendVector(A=@n93.R, B=@n93.R))
  n95 = TextureObjectParameter() {
   group=Wrap_Amber
   parameterName=Amber_Bug_Mask_2
   samplerType=SAMPLERTYPE_Masks
   texture=/Game/Weapons/WeaponSkins/Wraps/Textures/Amber/T_Wrap_Amber_M_2
  }
  n96 = Add(A=@n91, B=@n57)
  n97 = VectorParameter() {group=Wrap_Amber, parameterName=Amber_Bubbles_Speed_B, defaultValue=<rgba 0 0 0 1>}
  n98 = Multiply(A=Time(), B=AppendVector(A=@n97.R, B=@n97.R))
  n99 = Multiply(A=TextureSample(UVs=Add(A=@n92, B=@n94), Tex=@n95).R, B=TextureSample(UVs=Add(A=@n96, B=@n98), Tex=@n95).G)
  n100 = CurveAtlasRowParameter(CurveTime=@n78) {
   atlas=&CruveAtlas_Wraps_01
   curve=&C_Amber_Mask
   defaultValue=30
   group=Wrap_Amber
   parameterName=Amber_Mask_ColorCurve
  }
  n101 = ChannelMaskParameter(None=MaterialFunctionCall(
    Index=ScalarParameter() {defaultValue=30, group=Wrap_Amber, parameterName="Amber_Mask_ColorCurve  Index"}
    CCinput=@n100
    CCinput_Alpha=@n100
    UV=@n78
   ) {materialFunction=&MF_Global_CC_Tex}.4v) {group=Wrap_Amber, maskChannel=Alpha, parameterName=Amber_Mask_ColorCurve Channel Edge, defaultValue=<rgba 0 0 0 1>}
  n102 = ChannelMaskParameter(None=LinearInterpolate(A=@n13, B=NamedRerouteUsage(), Alpha=NamedRerouteUsage())) {group=Wrap_Amber, maskChannel=Blue, parameterName=Amber_FakeLight_Direction, defaultValue=<rgba 0 0 1 0>}
  n103 = Power(Base=@n102.R, Exp=ScalarParameter() {defaultValue=2.5, group=Wrap_Amber, parameterName=Amber_FakeLight_Falloff}) {constExponent=2.5}
  n104 = ScalarParameter() {group=Wrap_Amber, parameterName=Amber_FakeLight_Offset}
  n105 = ScalarParameter() {defaultValue=0.1, group=Wrap_Amber, parameterName=Amber_FakeLight_Warp_Normals}
  n106 = Add(A=@n101.R, B=Add(A=Add(A=@n103, B=@n104), B=Multiply(A=@n83, B=@n105) {constB=0.1})) {constB=0}
  n107 = CurveAtlasRowParameter(CurveTime=@n106) {
   atlas=&CruveAtlas_Wraps_01
   curve=&C_Amber_Mask
   defaultValue=30
   group=Wrap_Amber
   parameterName=Amber_Mask_ColorCurve
  }
  n108 = ChannelMaskParameter(None=MaterialFunctionCall(
    Index=ScalarParameter() {defaultValue=30, group=Wrap_Amber, parameterName="Amber_Mask_ColorCurve  Index"}
    CCinput=@n107
    CCinput_Alpha=@n107
    UV=@n106
   ) {materialFunction=&MF_Global_CC_Tex}.4v) {group=Wrap_Amber, parameterName=Amber_Mask_ColorCurve Channel Light}
  n109 = Multiply(A=@n83, B=ScalarParameter() {defaultValue=0.1, group=Wrap_Amber, parameterName=Amber_Inner_Warp_Normals})
  n110 = ScalarParameter() {defaultValue=1, group=Wrap_Amber, parameterName=Amber_Inner_Radus}
  n111 = ScalarParameter() {defaultValue=0.1, group=Wrap_Amber, parameterName=Amber_Inner_Radus_Mech}
  n112 = ScalarParameter() {group=Wrap_Amber, parameterName=Amber_Inner_Falloff}
  n113 = ScalarParameter() {group=Wrap_Amber, parameterName=Amber_Inner_Falloff_Mech}
  n114 = Power(
   Base=SphereMask(
    A=Add(A=@n109, B=@n35)
    B=Constant() {r=0.5}
    Radius=LinearInterpolate(A=@n110, B=@n111, Alpha=@n27.B)
    Hardness=LinearInterpolate(A=@n112, B=@n113, Alpha=@n27.B)
   )
   Exp=ScalarParameter() {defaultValue=2, group=Wrap_Amber, parameterName=Amber_Inner_Softern_Fallof}
  )
  n115 = CurveAtlasRowParameter(CurveTime=@n114) {
   atlas=&CruveAtlas_Wraps_01
   curve=&C_Amber_Mask
   defaultValue=30
   group=Wrap_Amber
   parameterName=Amber_Mask_ColorCurve
  }
  n116 = ChannelMaskParameter(None=MaterialFunctionCall(
    Index=ScalarParameter() {defaultValue=30, group=Wrap_Amber, parameterName="Amber_Mask_ColorCurve  Index"}
    CCinput=@n115
    CCinput_Alpha=@n115
    UV=@n114
   ) {materialFunction=&MF_Global_CC_Tex}.4v) {group=Wrap_Amber, maskChannel=Blue, parameterName=Amber_Mask_ColorCurve Channel Fallof, defaultValue=<rgba 0 0 1 0>}
  n117 = Multiply(A=Saturate(None=Add(A=@n90, B=@n99)), B=Multiply(A=@n108.R, B=@n116.R))
  n118 = Multiply(A=ScalarParameter() {group=Wrap_Amber, parameterName=Amber_Bug_Warp_Curvature}, B=@n83)
  n119 = Time()
  n120 = VectorParameter() {group=Wrap_Amber, parameterName=Amber_bug_Speed_A, defaultValue=<rgba 0 0 0 1>}
  n121 = AppendVector(A=@n120.R, B=@n120.R)
  n122 = TextureObjectParameter() {
   group=Wrap_Amber
   parameterName=Amber_Bug_Mask
   samplerType=SAMPLERTYPE_Masks
   texture=/Game/Weapons/WeaponSkins/Wraps/Textures/Amber/T_Wrap_Amber_M
  }
  n123 = TextureSample(UVs=Add(A=Add(A=@n118, B=@n46), B=Multiply(A=@n119, B=@n121)), Tex=@n122)
  n124 = Multiply(A=@n123.A, B=@n89.Result)
  n125 = Multiply(
   A=OneMinus(None=ComponentMask(None=@n21.Combined) {r=true})
   B=Multiply(A=TextureSample(UVs=@n84, Tex=@n1).B, B=LinearInterpolate(A=@n117, B=@n123.R, Alpha=@n124))
  )
  n126 = CurveAtlasRowParameter(CurveTime=@n125) {
   atlas=&CruveAtlas_Wraps_01
   curve=/Game/Weapons/WeaponSkins/Wraps/Textures/Amber/C_Amber_Color
   defaultValue=29
   group=Wrap_Amber
   parameterName=Amber_Edge_ColorCurve
  }
  n127 = Multiply(
   A=VectorParameter() {group=Wrap_Amber, parameterName=Amber_Base_Color, defaultValue=<rgba 0.7991 0.4287 0.0194 1>}.RGB
   B=MaterialFunctionCall(Index=ScalarParameter() {defaultValue=29, group=Wrap_Amber, parameterName="Amber_Edge_ColorCurve  Index"}, CCinput=@n126, UV=@n125) {materialFunction=&MF_Global_CC_Tex}.3V
  )
  n128 = Multiply(A=@n127, B=ScalarParameter() {defaultValue=0.5, group=Wrap_Amber, parameterName=Amber_Emissive_Amount, sortPriority=0})
  n129 = ScalarParameter() {defaultValue=0.5, group=Wrap_Amber_TOD, parameterName=Amber_TOD_Emis_Mornning}
  n130 = ScalarParameter() {defaultValue=1, group=Wrap_Amber_TOD, parameterName=ParamAmber_TOD_Emis_Day}
  n131 = ScalarParameter() {defaultValue=0.5, group=Wrap_Amber_TOD, parameterName=Amber_TOD_Emis_Evenning}
  n132 = ScalarParameter() {defaultValue=0.25, group=Wrap_Amber_TOD, parameterName=Amber_TOD_Emis_Night}
  n133 = TextureObjectParameter() {
   group=Wrap_Amber
   parameterName=Amber_BS_Normals
   samplerType=SAMPLERTYPE_Normal
   texture=/Game/Weapons/WeaponSkins/Wraps/Textures/Amber/T_Wrap_Amber_N
  }
  n134 = MaterialFunctionCall!(
   Texture_Normals=@n133
   Rotation_Axis=@n20
   Caculated_Bounds=@n15
   Pre_Skinned_Normals=@n7
   Blend_Contrast=@n2
   Fill_alternate_Axis_Opacity=@n3
   Is_Vehicle=@n6.R
   Is_Mech=@n6.R
  ) {materialFunction=&/MF_TopsSideProjection_Normals}
  n135 = GetMaterialAttributes(None=@n74) {attributeGetTypes [#1]}
  n136 = SetMaterialAttributes!(
   MaterialAttributes=@n75.MaterialAttributes
   BaseColor=LinearInterpolate(A=@n76.RGB, B=LinearInterpolate(B=@n77, Alpha=@n77) {constB=0}, Alpha=@n80.R)
   Roughness=LinearInterpolate(
    A=ScalarParameter() {group=Wrap_Amber, parameterName=Amber_Roghness_Low, sortPriority=0}
    B=ScalarParameter() {defaultValue=1, group=Wrap_Amber, parameterName=Amber_Roghness_High, sortPriority=0}
    Alpha=Saturate(None=@n81)
   )
   Specular=ScalarParameter() {defaultValue=0.5, group=Wrap_Amber, parameterName=Amber_Specular, sortPriority=0}
   EmissiveColor=Multiply(
    A=Add(A=@n82, B=@n128)
    B=MaterialFunctionCall(Value_Morning=@n129, Value_Day=@n130, Value_Evening=@n131, Value_Night=@n132) {materialFunction=&Scalar_Time_of_Day}.Result
   )
   Metallic=ScalarParameter() {group=Wrap_Amber, parameterName=Amber_Metallic, sortPriority=0}
   Normal=LinearInterpolate(
    A=@n135.Normal
    B=Add(A=@n135.Normal, B=@n134.Add_Surface_Normal)
    Alpha=ScalarParameter() {defaultValue=1, group=Wrap_Amber, parameterName=Amber_BS_Normal_Intensty, sortPriority=1}
   )
  ) {attributeSetTypes [
    #2
    #3
    #4
    #5
    #6
    #1
   ]}
  n137 = NamedRerouteUsage()
  n138 = NamedRerouteUsage()
  n139 = LinearInterpolate(
   B=FunctionInput(Preview=TextureSample() {samplerType=SAMPLERTYPE_Masks, texture=&T_Scar_CM}.R) {bUsePreviewValueAsDefault=true, inputName="Customization Mask (_CM B)", inputType=FunctionInput_Scalar, sortPriority=24}
   Alpha=Saturate(None=Subtract(A=@n137, B=@n138))
  ) {constA=1}
  FunctionOutput(None=StaticSwitchParameter(True=BlendMaterialAttributes(A=@n75.MaterialAttributes, B=@n136, Alpha=@n139) {vertexAttributeBlendType=UseA}, False=@n75.MaterialAttributes) {group=Customization_Wrap, parameterName=Wrap_Amber})
  LinearInterpolate(A=@n127, Alpha=@n80.R) {constA=1}
  MaterialFunctionCall() {materialFunction=&ObjectPivotPoint}
  n140 = TransformPosition(None=Constant3Vector()) {transformType=TRANSFORMPOSSOURCE_World}
  n141 = Normalize(VectorInput=Subtract(A=@n140, B=CameraPositionWS()))
  n142 = Constant3Vector() {constant=<rgba 0 0 1 1>}
  n143 = FunctionInput() {bUsePreviewValueAsDefault=true, inputName=Object-SafeZone-Virtual-Pivot, inputType=FunctionInput_Vector4, sortPriority=105}
  n144 = MaterialFunctionCall() {materialFunction=&ObjectScale}
  n145 = MaterialFunctionCall(Float3=PreSkinnedLocalBounds().Extents) {materialFunction=&BreakOutFloat3Components}
  n146 = Max(A=Max(A=@n145.R, B=@n145.R), B=@n145.B)
  MaterialFunctionCall(
   Plane Normal Axis=@n141
   Plane X Axis=Normalize(VectorInput=CrossProduct(A=@n141, B=@n142))
   Plane Center=Add(A=@n140, B=ComponentMask(None=@n143) {b=true, g=true, r=true})
   UV Scale=Multiply(A=ComponentMask(None=@n143) {a=true}, B=Multiply(A=@n144.Scale X, B=@n146))
  ) {materialFunction=&VirtualPlaneCoordinatesAbs}
  Multiply(A=@n80.R)
  Multiply(A=@n61, B=Constant2Vector() {g=2, r=1})
  Multiply(A=@n100.A)
  Multiply(A=@n83)
  Multiply(A=@n108.R)
  n147 = Constant()
  n148 = MaterialFunctionCall(Input Low=@n147, Input High=Constant() {r=0.5}, Target Low=@n147, Target High=Constant() {r=1}) {materialFunction=&RemapValueRange}
  Multiply(A=OneMinus(None=Saturate(None=@n148.Result))) {constB=0.125}
  NamedRerouteDeclaration(None=@n27.A) {name=IsCharacter, nodeColor=<rgba 0.8938 1 0 1>}
  NamedRerouteDeclaration(None=@n14) {name=Bounds, nodeColor=<rgba 1 0.5859 0 1>}
  n149 = FunctionInput() {bUsePreviewValueAsDefault=true, inputName=IsStatsB, inputType=FunctionInput_Vector2, sortPriority=-19}
  NamedRerouteDeclaration(None=ComponentMask(None=@n149) {r=true}) {name=IsBackpack, nodeColor=<rgba 0 0.9253 1 1>}
  NamedRerouteDeclaration(None=FunctionInput() {bUsePreviewValueAsDefault=true, inputName=isCharAccent, inputType=FunctionInput_Scalar, sortPriority=78}) {name=isCharAccent, nodeColor=<rgba 0 0.9253 1 1>}
  NamedRerouteDeclaration(None=ComponentMask(None=@n149) {g=true}) {name=IsPickaxe, nodeColor=<rgba 0 0.9253 1 1>}
  n150 = FunctionInput() {bUsePreviewValueAsDefault=true, inputName=Object-SafeZone-Virtual-CH-Switch, inputType=FunctionInput_Scalar, sortPriority=105}
  NamedRerouteDeclaration(None=MaterialFunctionCall!(
    Camera _pos
    Axis
    Scale
    Scale_Multplyer
    Object-SafeZone-Virtual-Pivot=@n143
    Object-SafeZone-Virtual-CH-Switch=@n150
    SinglePlane Normal Axis
    SinglePlane X Axis
    PlaneCenter_distortion
   ) {materialFunction=&/MF_SafeZone_VirtualPlane}.Uvs) {name=SafeZoneVP_UVs, nodeColor=<rgba 1 0 0.6562 1>}
  NamedRerouteUsage()
  Reroute()
  ScalarParameter() {defaultValue=1, group=Wrap_Amber, parameterName=Amber_Inner_Radus}
  ScalarParameter() {defaultValue=0.2, group=Wrap_Amber, parameterName=Amber_Roghness, sortPriority=0}
 ]}
