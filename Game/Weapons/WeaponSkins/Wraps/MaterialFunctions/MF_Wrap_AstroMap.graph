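# Material function for the AstroMap weapon-skin wrap, painting an animated solar system with per-type placement for weapons, vehicles, mechs and characters.
# Glitch-distorted side-projected UVs drive a scanlined system disc, six rotating, tinted planets and a "Zero" overlay.
# Projected glittering stars and curvature edge color join them in time-of-day-scaled emissive.
# Customization masks blend the result over the input material, behind the Wrap_AstroMap static switch; the glitch on/off pulse is unconnected.
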MaterialFunction /Game/Weapons/WeaponSkins/Wraps/MaterialFunctions/MF_Wrap_AstroMap
graph {nodes [
  Add() {constB=0}
  n1 = GetMaterialAttributes(None=FunctionInput() {bUsePreviewValueAsDefault=true, inputName=Material, inputType=FunctionInput_MaterialAttributes, sortPriority=-20}) {attributeGetTypes [#1]}
  n2 = Add(A=FunctionInput() {bUsePreviewValueAsDefault=true, inputName=Curavture, inputType=FunctionInput_Scalar, sortPriority=24}) {constB=-0.5}
  n3 = Multiply(A=VectorParameter() {parameterName=AstroMap_Edge_Color, defaultValue=<rgba 0 0 0 1>}.RGB, B=Saturate(None=Multiply(A=@n2) {constB=2}))
  n4 = FunctionInput(Preview=VertexInterpolator(VS=PreSkinnedNormal()).PS) {bUsePreviewValueAsDefault=true, inputName=Pre-Skinned Local Normal, sortPriority=10}
  n5 = MaterialFunctionCall(Color=Abs(None=@n4)) {materialFunction=&SplitComponents}
  n6 = FunctionInput(Preview=Constant()) {bUsePreviewValueAsDefault=true, inputName=isSates, inputType=FunctionInput_Vector4, sortPriority=4}
  n7 = MaterialFunctionCall(Color=@n6) {materialFunction=&SplitComponents}
  n8 = LinearInterpolate(A=@n5.B, B=@n5.B, Alpha=@n7.G)
  n9 = ScalarParameter() {defaultValue=1, group=Wrap_AstroMap, parameterName=AstroMap_Contrast}
  n10 = MaterialFunctionCall(In=LinearInterpolate(A=LinearInterpolate(A=@n5.R, B=@n8, Alpha=@n7.R), B=@n5.G, Alpha=@n7.B), Contrast=@n9) {materialFunction=&CheapContrast}
  n11 = VectorParameter() {group=Wrap_AstroMap, parameterName=AstroMap_SolarSys_ScanlinesColor, defaultValue=<rgba 0.1615 0.1615 0.1615 1>}
  n12 = Time()
  n13 = ScalarParameter() {defaultValue=-0.1, group=Wrap_AstroMap, parameterName=AstroMap_SolarSys_Scanline_Speed}
  n14 = VertexInterpolator(VS=PreSkinnedPosition())
  n15 = PreSkinnedLocalBounds()
  n16 = FunctionInput(Preview=Divide(A=Subtract(A=@n14.PS, B=@n15.Min), B=Subtract(A=@n15.Max, B=@n15.Max))) {bUsePreviewValueAsDefault=true, inputName=Calcated_Bounds, sortPriority=11}
  n17 = MaterialFunctionCall(Color=@n6) {materialFunction=&SplitComponents}
  n18 = MaterialFunctionCall!(
   For_Safe_Zones
   Normal_Map
   Caculated_Bounds=@n16
   Pre_Skinned_Normals=@n4
   Is Vehicle=@n17.R
   Is VehicleSide=@n17.R
   Is Mech=@n17.R
   Vehicle_Rotation
   UV_Rotation Bypass
   Mech_Rotation
   Uvs_Offset
   Offset-after rotation
   Uvs_TileRate
  ) {materialFunction=&/MF_Projection_Side_Normal}
  n19 = ScalarParameter() {defaultValue=-0.25, group=Wrap_AstroMap, parameterName=AstroMap_Glitch_Rotation_Vehicle}
  n20 = ScalarParameter() {group=Wrap_AstroMap, parameterName=AstroMap_Glitch_Rotation_VehicleSide}
  n21 = MaterialFunctionCall(Color=@n6) {materialFunction=&SplitComponents}
  n22 = LinearInterpolate(
   A=LinearInterpolate(
    A=ScalarParameter() {group=Wrap_AstroMap, parameterName=AstroMap_Glitch_Rotation_Weapon}
    B=LinearInterpolate(A=@n19, B=@n20, Alpha=@n21.G)
    Alpha=@n21.R
   )
   B=ScalarParameter() {group=Wrap_AstroMap, parameterName=AstroMap_Glitch_Rotation_Mech}
   Alpha=@n21.B
  )
  n23 = Add(
   A=Multiply(
    A=ScalarParameter() {defaultValue=1, group=Wrap_AstroMap, parameterName=AstroMap_Glitch_Scale}
    B=MaterialFunctionCall(UVs=@n18.UVs No Offsets/Tiling, "Rotation Angle (0-1)"=@n22) {materialFunction=&CustomRotator}.Rotated Values
   )
   B=Multiply(A=Time(), B=ScalarParameter() {defaultValue=-0.05, group=Wrap_AstroMap, parameterName=AstroMap_Glitch_Speed})
  )
  n24 = TextureObjectParameter() {group=Wrap_AstroMap, parameterName=AstroMap_MaskC__Tex, texture=&T_BlacksRGB}
  n25 = Multiply(
   A=Add(A=TextureSample(UVs=@n23, Tex=@n24).A, B=ScalarParameter() {defaultValue=-0.6, group=Wrap_AstroMap, parameterName=AstroMap_Glitch_Offset}) {constB=-0.6}
   B=ScalarParameter() {defaultValue=0.05, group=Wrap_AstroMap, parameterName=AstroMap_Glitch_Strength}
  )
  n26 = VectorParameter() {group=Wrap_AstroMap, parameterName=AstroMap_SolarSys_Offset_Vehicle, defaultValue=<rgba 0 0 0 1>}
  n27 = VectorParameter() {group=Wrap_AstroMap, parameterName=AstroMap_SolarSys_Offset_VehicleSide, defaultValue=<rgba 0 1 0 1>}
  n28 = MaterialFunctionCall(Float4=@n6) {materialFunction=&BreakOutFloat4Components}
  n29 = LinearInterpolate(
   A=LinearInterpolate(
    A=VectorParameter() {group=Wrap_AstroMap, parameterName=AstroMap_SolarSys_Offset_Weapon, defaultValue=<rgba 0 1 0 0>}.RGB
    B=LinearInterpolate(A=@n26.RGB, B=@n27.RGB, Alpha=@n28.G)
    Alpha=@n28.R
   )
   B=LinearInterpolate(
    A=VectorParameter() {group=Wrap_AstroMap, parameterName=AstroMap_SolarSys_Offset_Mech, defaultValue=<rgba 0 1 0 1>}.RGB
    B=VectorParameter() {group=Wrap_AstroMap, parameterName=AstroMap_SolarSys_Offset_Character, defaultValue=<rgba 0 1 0 1>}.RGB
    Alpha=@n28.A
   )
   Alpha=@n28.B
  )
  n30 = ScalarParameter() {defaultValue=0.5, group=Wrap_AstroMap, parameterName=AstroMap_Master_Scale_Weapon}
  n31 = MaterialFunctionCall(Float4=@n6) {materialFunction=&BreakOutFloat4Components}
  n32 = LinearInterpolate(
   A=ScalarParameter() {defaultValue=1, group=Wrap_AstroMap, parameterName=AstroMap_Master_Scale_Vehicle}
   B=ScalarParameter() {defaultValue=1, group=Wrap_AstroMap, parameterName=AstroMap_Master_Scale_VehicleSide}
   Alpha=@n31.G
  )
  n33 = ScalarParameter() {defaultValue=1, group=Wrap_AstroMap, parameterName=AstroMap_Master_Scale_Mech}
  n34 = ScalarParameter() {defaultValue=1, group=Wrap_AstroMap, parameterName=AstroMap_Master_Scale_Character}
  n35 = MaterialFunctionCall(
   UVs=Add(A=Add(A=@n25, B=@n18.UVs No Offsets/Tiling), B=ComponentMask(None=@n29) {g=true, r=true}) {constB=0}
   Texture Scale=LinearInterpolate(A=LinearInterpolate(A=@n30, B=@n32, Alpha=@n31.R), B=LinearInterpolate(A=@n33, B=@n34, Alpha=@n31.A), Alpha=@n31.B)
  ) {materialFunction=&ScaleUVsByCenter}
  n36 = Time()
  n37 = ScalarParameter() {defaultValue=-0.05, group=Wrap_AstroMap, parameterName=AstroMap_SolarSys_Rotation_Speed}
  n38 = MaterialFunctionCall(
   UVs=MaterialFunctionCall(UVs=@n35.UVs, "Rotation Angle (0-1)"=Multiply(A=@n36, B=@n37)) {materialFunction=&CustomRotator}.Rotated Values
   Texture Scale=ScalarParameter() {defaultValue=0.5, group=Wrap_AstroMap, parameterName=AstroMap_SolarSys_Scale}
  ) {materialFunction=&ScaleUVsByCenter}
  n39 = TextureObjectParameter() {group=Wrap_AstroMap, parameterName=AstroMap_MaskA__Tex, texture=&T_BlacksRGB}
  n40 = SphereMask(A=@n38.UVs, B=Constant() {r=0.5}, Radius=Constant() {r=0.5}, Hardness=Constant() {r=1})
  n41 = MaterialFunctionCall(Float4=Multiply(A=TextureSample(UVs=@n38.UVs, Tex=@n39).RGBA, B=@n40)) {materialFunction=&BreakOutFloat4Components}
  n42 = TextureSample(UVs=Add(A=Multiply(A=@n12, B=@n13), B=@n41.A), Tex=@n24)
  n43 = VectorParameter() {group=Wrap_AstroMap, parameterName=AstroMap_SolarSys_RingColor, defaultValue=<rgba 1 1 1 1>}
  n44 = VectorParameter() {group=Wrap_AstroMap, parameterName=AstroMap_SolarSys_BCColor A, defaultValue=<rgba 0.0332 0.004 0.0573 1>}
  n45 = VectorParameter() {group=Wrap_AstroMap, parameterName=AstroMap_SolarSys_BCColor B, defaultValue=<rgba 0.0887 0.0091 0.1559 1>}
  n46 = LinearInterpolate(A=@n44.RGB, B=@n45.RGB, Alpha=@n41.B)
  n47 = MaterialFunctionCall(
   UVs=@n35.UVs
   "Rotation Angle (0-1)"=Multiply(A=Time(), B=ScalarParameter() {defaultValue=0.05, group=Wrap_AstroMap, parameterName=AstroMap_PlanetA_Rotation_Speed})
  ) {materialFunction=&CustomRotator}
  n48 = ComponentMask(None=VectorParameter() {group=Wrap_AstroMap, parameterName=AstroMap_PlanetA_Offset, defaultValue=<rgba 0 0.24 0 1>}.RGB) {g=true, r=true}
  n49 = TextureObjectParameter() {group=Wrap_AstroMap, parameterName=AstroMap_MaskB__Tex, texture=&T_BlacksRGB}
  n50 = TextureSample(
   UVs=MaterialFunctionCall(
    UVs=Add(A=@n47.Rotated Values, B=@n48) {constB=0}
    Texture Scale=ScalarParameter() {defaultValue=0.1, group=Wrap_AstroMap, parameterName=AstroMap_PlanetA_Scale}
   ) {materialFunction=&ScaleUVsByCenter}.UVs
   Tex=@n49
  ) {samplerType=SAMPLERTYPE_LinearColor}
  n51 = OneMinus(None=@n50.A)
  n52 = Multiply(A=Add(A=Multiply(A=@n43.RGB, B=@n41.R), B=Multiply(A=@n41.G, B=@n46)), B=Saturate(None=Add(A=@n51) {constA=0.2, constB=0.1}))
  n53 = VectorParameter() {group=Wrap_AstroMap, parameterName=AstroMap_PlanetA_GlowColor, defaultValue=<rgba 0.0938 0.0938 0.0938 1>}
  n54 = VectorParameter() {group=Wrap_AstroMap, parameterName=AstroMap_PlanetA_BCColor A, defaultValue=<rgba 0.1549 0.1562 0.012 1>}
  n55 = VectorParameter() {group=Wrap_AstroMap, parameterName=AstroMap_PlanetA_BCColor B, defaultValue=<rgba 0.9823 0.9911 0.0762 1>}
  n56 = LinearInterpolate(A=@n54.RGB, B=@n55.RGB, Alpha=@n50.G)
  n57 = Add(
   A=Multiply(A=VectorParameter() {group=Wrap_AstroMap, parameterName=AstroMap_PlanetA_RingColor, defaultValue=<rgba 1 0.8512 0 1>}.RGB, B=@n50.R)
   B=LinearInterpolate(A=Multiply(A=@n53.RGB, B=@n50.B), B=Multiply(A=@n56, B=@n50.A), Alpha=@n50.A)
  )
  n58 = MaterialFunctionCall(
   UVs=@n35.UVs
   "Rotation Angle (0-1)"=Multiply(A=Time(), B=ScalarParameter() {defaultValue=-0.05, group=Wrap_AstroMap, parameterName=AstroMap_PlanetB_Rotation_Speed})
  ) {materialFunction=&CustomRotator}
  n59 = ComponentMask(None=VectorParameter() {group=Wrap_AstroMap, parameterName=AstroMap_PlanetB_Offset, defaultValue=<rgba 0 0.26 0 1>}.RGB) {g=true, r=true}
  n60 = TextureSample(
   UVs=MaterialFunctionCall(
    UVs=Add(A=@n58.Rotated Values, B=@n59) {constB=0}
    Texture Scale=ScalarParameter() {defaultValue=0.23, group=Wrap_AstroMap, parameterName=AstroMap_PlanetB_Scale}
   ) {materialFunction=&ScaleUVsByCenter}.UVs
   Tex=@n49
  ) {samplerType=SAMPLERTYPE_LinearColor}
  n61 = OneMinus(None=@n60.A)
  n62 = Add(A=@n61) {constA=0.2, constB=0.1}
  n63 = VectorParameter() {group=Wrap_AstroMap, parameterName=AstroMap_PlanetB_RingColor, defaultValue=<rgba 0.7835 0.3712 0.9647 1>}
  n64 = VectorParameter() {group=Wrap_AstroMap, parameterName=AstroMap_PlanetB_GlowColor, defaultValue=<rgba 0.0573 0.0573 0.0573 1>}
  n65 = Multiply(A=@n64.RGB, B=@n60.B)
  n66 = VectorParameter() {group=Wrap_AstroMap, parameterName=AstroMap_PlanetB_BCColor A, defaultValue=<rgba 0.1875 0.0007 0.1811 1>}
  n67 = VectorParameter() {group=Wrap_AstroMap, parameterName=AstroMap_PlanetB_BCColor B, defaultValue=<rgba 0.2723 0.1268 0.3385 1>}
  n68 = Multiply(A=LinearInterpolate(A=@n66.RGB, B=@n67.RGB, Alpha=@n60.G), B=@n60.A)
  n69 = Add(
   A=Multiply(A=Add(A=@n52, B=@n57), B=Saturate(None=@n62))
   B=Add(A=Multiply(A=@n63.RGB, B=@n60.R), B=LinearInterpolate(A=@n65, B=@n68, Alpha=@n60.A))
  )
  n70 = MaterialFunctionCall(
   UVs=@n35.UVs
   "Rotation Angle (0-1)"=Multiply(A=Time(), B=ScalarParameter() {defaultValue=0.085, group=Wrap_AstroMap, parameterName=AstroMap_PlanetE_Rotation_Speed})
  ) {materialFunction=&CustomRotator}
  n71 = ComponentMask(None=VectorParameter() {group=Wrap_AstroMap, parameterName=AstroMap_PlanetE_Offset, defaultValue=<rgba 0 0.23 0 1>}.RGB) {g=true, r=true}
  n72 = TextureSample(
   UVs=MaterialFunctionCall(
    UVs=Add(A=@n70.Rotated Values, B=@n71) {constB=0}
    Texture Scale=ScalarParameter() {defaultValue=0.2, group=Wrap_AstroMap, parameterName=AstroMap_PlanetE_Scale}
   ) {materialFunction=&ScaleUVsByCenter}.UVs
   Tex=@n49
  ) {samplerType=SAMPLERTYPE_LinearColor}
  n73 = OneMinus(None=@n72.A)
  n74 = Saturate(None=Add(A=@n73) {constA=0.2, constB=0.1})
  n75 = VectorParameter() {group=Wrap_AstroMap, parameterName=AstroMap_PlanetE_RingColor, defaultValue=<rgba 0.9597 0.9734 0.2791 1>}
  n76 = Multiply(A=@n75.RGB, B=@n72.R)
  n77 = VectorParameter() {group=Wrap_AstroMap, parameterName=AstroMap_PlanetE_GlowColor, defaultValue=<rgba 0.0573 0.0573 0.0573 1>}
  n78 = VectorParameter() {group=Wrap_AstroMap, parameterName=AstroMap_PlanetE_BCColor A, defaultValue=<rgba 0.0562 0.2031 0.0058 1>}
  n79 = VectorParameter() {group=Wrap_AstroMap, parameterName=AstroMap_PlanetE_BCColor B, defaultValue=<rgba 0.3682 0.7604 0.2337 1>}
  n80 = LinearInterpolate(A=Multiply(A=@n77.RGB, B=@n72.B), B=Multiply(A=LinearInterpolate(A=@n78.RGB, B=@n79.RGB, Alpha=@n72.G), B=@n72.A), Alpha=@n72.A)
  n81 = MaterialFunctionCall(
   UVs=@n35.UVs
   "Rotation Angle (0-1)"=Multiply(A=Time(), B=ScalarParameter() {defaultValue=-0.15, group=Wrap_AstroMap, parameterName=AstroMap_PlanetF_Rotation_Speed})
  ) {materialFunction=&CustomRotator}
  n82 = ComponentMask(None=VectorParameter() {group=Wrap_AstroMap, parameterName=AstroMap_PlanetF_Offset, defaultValue=<rgba 0 0.14 0 1>}.RGB) {g=true, r=true}
  n83 = TextureSample(
   UVs=MaterialFunctionCall(
    UVs=Add(A=@n81.Rotated Values, B=@n82) {constB=0}
    Texture Scale=ScalarParameter() {defaultValue=0.11, group=Wrap_AstroMap, parameterName=AstroMap_PlanetF_Scale}
   ) {materialFunction=&ScaleUVsByCenter}.UVs
   Tex=@n49
  ) {samplerType=SAMPLERTYPE_LinearColor}
  n84 = OneMinus(None=@n83.A)
  n85 = Multiply(A=Add(A=Multiply(A=@n69, B=@n74), B=Add(A=@n76, B=@n80)), B=Saturate(None=Add(A=@n84) {constA=0.2, constB=0.1}))
  n86 = VectorParameter() {group=Wrap_AstroMap, parameterName=AstroMap_PlanetF_RingColor, defaultValue=<rgba 0.9597 0.9734 0.2791 1>}
  n87 = VectorParameter() {group=Wrap_AstroMap, parameterName=AstroMap_PlanetF_GlowColor, defaultValue=<rgba 0.0573 0.0573 0.0573 1>}
  n88 = VectorParameter() {group=Wrap_AstroMap, parameterName=AstroMap_PlanetF_BCColor A, defaultValue=<rgba 0.5677 0.1467 0.0065 1>}
  n89 = VectorParameter() {group=Wrap_AstroMap, parameterName=AstroMap_PlanetF_BCColor B, defaultValue=<rgba 0.9647 0.4508 0.0865 1>}
  n90 = LinearInterpolate(A=@n88.RGB, B=@n89.RGB, Alpha=@n83.G)
  n91 = Add(A=Multiply(A=@n86.RGB, B=@n83.R), B=LinearInterpolate(A=Multiply(A=@n87.RGB, B=@n83.B), B=Multiply(A=@n90, B=@n83.A), Alpha=@n83.A))
  n92 = MaterialFunctionCall(
   UVs=@n35.UVs
   "Rotation Angle (0-1)"=Multiply(A=Time(), B=ScalarParameter() {defaultValue=-0.11, group=Wrap_AstroMap, parameterName=AstroMap_PlanetC_Rotation_Speed})
  ) {materialFunction=&CustomRotator}
  n93 = ComponentMask(None=VectorParameter() {group=Wrap_AstroMap, parameterName=AstroMap_PlanetC_Offset, defaultValue=<rgba 0.093 0 0 1>}.RGB) {g=true, r=true}
  n94 = TextureSample(
   UVs=MaterialFunctionCall(
    UVs=Add(A=@n92.Rotated Values, B=@n93) {constB=0}
    Texture Scale=ScalarParameter() {defaultValue=0.4, group=Wrap_AstroMap, parameterName=AstroMap_PlanetC_Scale}
   ) {materialFunction=&ScaleUVsByCenter}.UVs
   Tex=@n49
  ) {samplerType=SAMPLERTYPE_LinearColor}
  n95 = Add(A=OneMinus(None=@n94.A)) {constA=0.2, constB=0.1}
  n96 = VectorParameter() {group=Wrap_AstroMap, parameterName=AstroMap_PlanetC_RingColor, defaultValue=<rgba 0.4798 0.2389 0.6042 1>}
  n97 = VectorParameter() {group=Wrap_AstroMap, parameterName=AstroMap_PlanetC_GlowColor, defaultValue=<rgba 0.0573 0.0573 0.0573 1>}
  n98 = Multiply(A=@n97.RGB, B=@n94.B)
  n99 = VectorParameter() {group=Wrap_AstroMap, parameterName=AstroMap_PlanetC_BCColor B, defaultValue=<rgba 0.9062 0.1497 0.0503 1>}
  n100 = Multiply(
   A=LinearInterpolate(
    A=VectorParameter() {group=Wrap_AstroMap, parameterName=AstroMap_PlanetC_BCColor A, defaultValue=<rgba 0.4323 0.0215 0 1>}.RGB
    B=@n99.RGB
    Alpha=@n94.G
   )
   B=@n94.A
  )
  n101 = Add(
   A=Multiply(A=Add(A=@n85, B=@n91), B=Saturate(None=@n95))
   B=Add(A=Multiply(A=@n96.RGB, B=@n94.R), B=LinearInterpolate(A=@n98, B=@n100, Alpha=@n94.A))
  )
  n102 = MaterialFunctionCall(
   UVs=@n35.UVs
   "Rotation Angle (0-1)"=Multiply(A=Time(), B=ScalarParameter() {defaultValue=0.085, group=Wrap_AstroMap, parameterName=AstroMap_PlanetD_Rotation_Speed})
  ) {materialFunction=&CustomRotator}
  n103 = ComponentMask(None=VectorParameter() {group=Wrap_AstroMap, parameterName=AstroMap_PlanetD_Offset, defaultValue=<rgba 0 0.045 0 1>}.RGB) {g=true, r=true}
  n104 = TextureSample(
   UVs=MaterialFunctionCall(
    UVs=Add(A=@n102.Rotated Values, B=@n103) {constB=0}
    Texture Scale=ScalarParameter() {defaultValue=0.3, group=Wrap_AstroMap, parameterName=AstroMap_PlanetD_Scale}
   ) {materialFunction=&ScaleUVsByCenter}.UVs
   Tex=@n49
  ) {samplerType=SAMPLERTYPE_LinearColor}
  n105 = Saturate(None=Add(A=OneMinus(None=@n104.A)) {constA=0.2, constB=0.1})
  n106 = VectorParameter() {group=Wrap_AstroMap, parameterName=AstroMap_PlanetD_RingColor, defaultValue=<rgba 0.3095 0.956 0.9734 1>}
  n107 = Multiply(A=@n106.RGB, B=@n104.R)
  n108 = VectorParameter() {group=Wrap_AstroMap, parameterName=AstroMap_PlanetD_GlowColor, defaultValue=<rgba 0.0573 0.0573 0.0573 1>}
  n109 = VectorParameter() {group=Wrap_AstroMap, parameterName=AstroMap_PlanetD_BCColor A, defaultValue=<rgba 0.0023 0.1498 0.2448 1>}
  n110 = VectorParameter() {group=Wrap_AstroMap, parameterName=AstroMap_PlanetD_BCColor B, defaultValue=<rgba 0.1516 0.7696 0.7865 1>}
  n111 = LinearInterpolate(A=Multiply(A=@n108.RGB, B=@n104.B), B=Multiply(A=LinearInterpolate(A=@n109.RGB, B=@n110.RGB, Alpha=@n104.G), B=@n104.A), Alpha=@n104.A)
  n112 = Add(A=Multiply(A=@n11.RGB, B=Multiply(A=@n42.R, B=@n41.G)), B=Add(A=Multiply(A=@n101, B=@n105), B=Add(A=@n107, B=@n111)))
  n113 = LinearInterpolate(B=Multiply(A=OneMinus(None=@n40), B=Multiply(A=@n61, B=@n73)), Alpha=@n10.Result) {constA=1}
  n114 = MaterialFunctionCall(Color=@n6) {materialFunction=&SplitComponents}
  n115 = LinearInterpolate(
   A=ScalarParameter() {defaultValue=1, group=Wrap_AstroMap, parameterName=AstroMap_Stars_Scale_Weapon}
   B=LinearInterpolate(
    A=ScalarParameter() {defaultValue=1, group=Wrap_AstroMap, parameterName=AstroMap_Stars_Scale_Vehicle}
    B=ScalarParameter() {defaultValue=1, group=Wrap_AstroMap, parameterName=AstroMap_Stars_Scale_VehicleSide}
    Alpha=@n114.G
   )
   Alpha=@n114.R
  )
  n116 = ScalarParameter() {defaultValue=1, group=Wrap_AstroMap, parameterName=AstroMap_Stars_Scale_Mech}
  n117 = MaterialFunctionCall(Color=@n6) {materialFunction=&SplitComponents}
  n118 = MaterialFunctionCall!(
   IsCharacter
   IsStatsB
   Texture=@n24
   Rotation_Axis=Constant3Vector()
   Caculated_Bounds=Multiply(A=@n16, B=LinearInterpolate(A=@n115, B=@n116, Alpha=@n114.B))
   Pre_Skinned_Normals=@n4
   Blend_Contrast=@n9
   Fill_alternate_Axis_Opacity=Constant() {r=1}
   "Fill_alternate_Axis_Color_A=Roughness"=Constant4Vector().RGBA
   Is_Vehicle=@n117.R
   Is_Mech=@n117.R
   Enable_Fill_Z_Weapon
   Fill_Z_Weapon_opacity
   Fill_Z_Weapon_Color
  ) {materialFunction=&/MF_TopsSideProjection_Base}
  n119 = ComponentMask(None=@n118.Combined) {b=true}
  n120 = Multiply(A=Time(), B=ScalarParameter() {defaultValue=0.1, group=Wrap_AstroMap, parameterName=AstroMap_Stars_Glitter_Speed})
  n121 = Add(
   A=Sine(None=Add(A=@n119, B=@n120))
   B=ScalarParameter() {defaultValue=0.75, group=Wrap_AstroMap, parameterName=AstroMap_Stars_Glitter_RangeA}
  )
  n122 = ScalarParameter() {defaultValue=0.75, group=Wrap_AstroMap, parameterName=AstroMap_Stars_Glitter_RangeB}
  n123 = Multiply(A=ComponentMask(None=@n118.Combined) {g=true}, B=Saturate(None=Multiply(A=@n121, B=@n122)))
  n124 = Multiply(A=Time(), B=ScalarParameter() {defaultValue=-0.05, group=Wrap_AstroMap, parameterName=AstroMap_Zero_Rotation_Speed})
  n125 = TextureSample(
   UVs=MaterialFunctionCall(
    UVs=MaterialFunctionCall(UVs=@n35.UVs, "Rotation Angle (0-1)"=@n124) {materialFunction=&CustomRotator}.Rotated Values
    Texture Scale=ScalarParameter() {defaultValue=0.23, group=Wrap_AstroMap, parameterName=AstroMap_Zero_Scale}
   ) {materialFunction=&ScaleUVsByCenter}.UVs
   Tex=TextureObjectParameter() {group=Wrap_AstroMap, parameterName=AstroMap_Zero__Tex, texture=&T_BlacksRGB}
  ) {samplerType=SAMPLERTYPE_LinearColor}
  n126 = LinearInterpolate(A=Add(A=Multiply(A=@n10.Result, B=@n112), B=Multiply(A=@n113, B=@n123)), B=@n125.RGB, Alpha=Multiply(A=@n10.Result, B=@n125.A))
  n127 = Constant() {r=0.5}
  n128 = Constant() {r=1}
  n129 = Constant() {r=0.5}
  n130 = Constant() {r=0.25}
  n131 = SetMaterialAttributes!(
   MaterialAttributes=@n1.MaterialAttributes
   BaseColor=VectorParameter() {group=Wrap_AstroMap, parameterName=AstroMap_BaseColor, defaultValue=<rgba 0 0 0 1>}.RGB
   Roughness=ScalarParameter() {defaultValue=0.2, group=Wrap_AstroMap, parameterName=AstroMap_Roghness, sortPriority=0}
   Specular=ScalarParameter() {defaultValue=0.5, group=Wrap_AstroMap, parameterName=AstroMap_Specular, sortPriority=0}
   EmissiveColor=Multiply(
    A=Add(A=@n3, B=@n126)
    B=MaterialFunctionCall(Value_Morning=@n127, Value_Day=@n128, Value_Evening=@n129, Value_Night=@n130) {materialFunction=&Scalar_Time_of_Day}.Result
   )
   Metallic=ScalarParameter() {group=Wrap_AstroMap, parameterName=AstroMap_Metallic, sortPriority=0}
  ) {attributeSetTypes [
    #2
    #3
    #4
    #5
    #6
   ]}
  n132 = TextureSample() {samplerType=SAMPLERTYPE_Masks, texture=/Game/Weapons/FORT_Rifles/Textures/T_Scar_CM}
  n133 = Saturate(None=Add(A=@n132.R, B=@n132.R))
  n134 = Subtract(A=NamedRerouteUsage(), B=NamedRerouteUsage())
  n135 = Saturate(None=LinearInterpolate(
    A=FunctionInput(Preview=@n133) {bUsePreviewValueAsDefault=true, inputName=Customization Mask A, inputType=FunctionInput_Scalar, sortPriority=24}
    B=FunctionInput() {bUsePreviewValueAsDefault=true, inputName=Customization Mask B, inputType=FunctionInput_Scalar, sortPriority=24}
    Alpha=Saturate(None=@n134)
   ))
  FunctionOutput(None=StaticSwitchParameter(True=BlendMaterialAttributes(A=@n1.MaterialAttributes, B=@n131, Alpha=@n135) {vertexAttributeBlendType=UseA}, False=@n1.MaterialAttributes) {group=Customization_Wrap, parameterName=Wrap_AstroMap})
  n136 = Time()
  n137 = ScalarParameter() {defaultValue=0.5, group=Wrap_AstroMap, parameterName=AstroMap_Glitch_On_Speed}
  n138 = Sine(None=Add(B=Multiply(A=@n136, B=@n137))) {period=1.25}
  n139 = ScalarParameter() {defaultValue=-0.2, group=Wrap_AstroMap, parameterName=AstroMap_Glitch_On_RangeA}
  n140 = Saturate(None=Multiply(A=Add(A=@n138, B=@n139), B=ScalarParameter() {defaultValue=1.3, group=Wrap_AstroMap, parameterName=AstroMap_Glitch_On_RangeB}))
  Multiply(A=@n25, B=@n140)
  Multiply(A=MaterialFunctionCall!(X range Y range, Gradient Height, ShowGrid?, Thickness, Color, "f(x)"=@n140) {materialFunction=/Engine/Functions/Engine_MaterialFunctions01/Debug/PlotFunctionOnGraph}."f(x) = y")
  NamedRerouteDeclaration(None=MaterialFunctionCall(Float4=@n6) {materialFunction=&BreakOutFloat4Components}.A) {name=isCharacter, nodeColor=<rgba 0.4062 1 0 1>}
  NamedRerouteDeclaration(None=FunctionInput() {bUsePreviewValueAsDefault=true, inputName=isCharAccent, inputType=FunctionInput_Scalar, sortPriority=78}) {name=isCharAccent, nodeColor=<rgba 0 0.9253 1 1>}
 ]}
