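# Material function for the BSequins cosmetic wrap, blending a metallic sequin layer over incoming material attributes by customization mask, behind the Wrap_BSequins static switch.
# Tiled textures supply sequin mask, roughness and normals, with normal strength varying by pre-skinned normal Z.
# A sine-banded view-angle term drives the two-color falloff.
# Emissive sparkle multiplies panning noise with smoke projected by WorldAlignedTexture using pre-skinned position and normals, anchoring it to the mesh.
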MaterialFunction /Game/Weapons/WeaponSkins/Wraps/MaterialFunctions/MF_Wrap_BSequins
graph {nodes [
  Constant()
  n1 = Constant()
  FunctionInput(Preview=@n1) {bUsePreviewValueAsDefault=true, inputName=Side Project Vehicle, inputType=FunctionInput_Scalar}
  FunctionInput(Preview=@n1) {bUsePreviewValueAsDefault=true, inputName=isMech, inputType=FunctionInput_Scalar, sortPriority=4}
  FunctionInput(Preview=VertexInterpolator(VS=PreSkinnedNormal()).PS) {bUsePreviewValueAsDefault=true, inputName=Pre-Skinned Local Normal, sortPriority=10}
  FunctionInput(Preview=Constant()) {inputName=isVehicle, inputType=FunctionInput_Scalar, sortPriority=5}
  n2 = GetMaterialAttributes(None=FunctionInput(Preview=SetMaterialAttributes!(MaterialAttributes)) {inputName=Material, inputType=FunctionInput_MaterialAttributes, sortPriority=-20})
  n3 = NamedRerouteUsage()
  n4 = Saturate(None=DotProduct(A=PixelNormalWS(), B=CameraVectorWS()))
  n5 = ConstantBiasScale(None=Sine(None=@n4) {period=0.8})
  n6 = LinearInterpolate(Alpha=Multiply(A=@n5, B=OneMinus(None=@n4))) {constA=0.7}
  n7 = TextureSample(UVs=NamedRerouteUsage()) {
   samplerSource=SSM_Wrap_WorldGroupSettings
   samplerType=SAMPLERTYPE_LinearGrayscale
   texture=/Game/Global/Textures/Default/Tileable_Patterns/T_SequinTile_roughness
  }
  n8 = Time()
  n9 = ScalarParameter() {defaultValue=0.01, group=Wrap_BSequins, parameterName=BS_Sparkle_Speed}
  n10 = TextureSample(UVs=Add(A=NamedRerouteUsage(), B=Multiply(A=@n8, B=@n9) {constB=0.2})) {samplerSource=SSM_Wrap_WorldGroupSettings, samplerType=SAMPLERTYPE_Normal, texture=/Game/Global/Textures/Default/Noise/64x64Noise}
  n11 = VertexInterpolator(VS=PreSkinnedPosition())
  n12 = Multiply(A=VectorParameter() {group=Wrap_BSequins, parameterName=BS_Sparkle_Noise_Pan_Speed, defaultValue=<rgba 5 5 10 1>}.RGB, B=Time())
  n13 = VertexInterpolator(VS=PreSkinnedNormal())
  n14 = Saturate(None=MaterialFunctionCall(
    TextureObject=TextureObject() {texture=&T_Inky_Smoke_Tile}
    WorldPosition=Add(A=@n11.PS, B=@n12)
    World Space Normal=@n13.PS
   ) {materialFunction=&WorldAlignedTexture}.XY Texture)
  n15 = Multiply(
   A=Multiply(A=Power(Base=@n10.R) {constExponent=4}, B=Power(Base=@n14))
   B=VectorParameter() {group=Wrap_BSequins, parameterName=BS_Sparkle_Color, defaultValue=<rgba 0.5742 0.228 1 1>}.RGB
  )
  n16 = ScalarParameter() {defaultValue=10, group=Wrap_BSequins, parameterName=BS_Sparkle_Brightness}
  n17 = Constant3Vector() {constant=<rgba 0 0 1 0>}
  n18 = TextureSample(
   UVs=NamedRerouteUsage()
   Tex=TextureObjectParameter() {group=Wrap_BSequins, parameterName=BS_Normals, texture=&T_BlacksRGB}
  )
  n19 = DotProduct(A=@n13.PS, B=Constant3Vector() {constant=<rgba 0 0 1 0>})
  n20 = Multiply(
   A=ScalarParameter() {defaultValue=0.8, group=Wrap_BSequins, parameterName=BS_Normal_Intensity}
   B=Saturate(None=ConstantBiasScale(None=@n19))
  )
  n21 = SetMaterialAttributes!(
   MaterialAttributes=@n2.MaterialAttributes
   BaseColor=Multiply(A=ScalarParameter() {defaultValue=1.5, group=Wrap_BSequins, parameterName=BS_Brightness}, B=Multiply(A=@n3, B=@n6))
   Metallic=ScalarParameter() {defaultValue=0.9, group=Wrap_BSequins, parameterName=BS_Metallic}
   Roughness=@n7.R
   EmissiveColor=Multiply(A=NamedRerouteUsage(), B=Multiply(A=@n15, B=@n16))
   Normal=MaterialFunctionCall(BaseNormal=Constant3Vector() {constant=<rgba 0 0 1 0>}, AdditionalNormal=LinearInterpolate(A=@n17, B=@n18.RGB, Alpha=@n20)) {materialFunction=&BlendAngleCorrectedNormals}.Result
  ) {attributeSetTypes [
    #1
    #2
    #3
    #4
    #5
   ]}
  n22 = Saturate(None=Add(
    A=FunctionInput() {bUsePreviewValueAsDefault=true, inputName="Customization Mask (_CM B)", inputType=FunctionInput_Scalar, sortPriority=24}
    B=ScalarParameter() {group=Wrap_BSequins, parameterName=BS_Remove_Customization Mask, sortPriority=-10}
   ))
  FunctionOutput(None=StaticSwitchParameter(True=BlendMaterialAttributes(A=@n2.MaterialAttributes, B=@n21, Alpha=@n22) {vertexAttributeBlendType=UseA}, False=@n2.MaterialAttributes) {group=Customization_Wrap, parameterName=Wrap_BSequins})
  n23 = PreSkinnedLocalBounds()
  n24 = Subtract(A=VertexInterpolator(VS=PreSkinnedPosition()).PS, B=@n23.Min)
  n25 = Subtract(A=@n23.Max, B=@n23.Max)
  n26 = FunctionInput(Preview=Constant()) {bUsePreviewValueAsDefault=true, inputName=IsState, inputType=FunctionInput_Vector4}
  n27 = MaterialFunctionCall(Color=@n26) {materialFunction=&SplitComponents}
  MaterialFunctionCall!(
   For_Safe_Zones
   Normal_Map
   Caculated_Bounds=FunctionInput(Preview=Divide(A=@n24, B=@n25)) {bUsePreviewValueAsDefault=true, inputName=Calcated_Bounds, sortPriority=11}
   Pre_Skinned_Normals
   Is Vehicle=@n27.R
   Is VehicleSide=@n27.R
   Is Mech=@n27.R
   Vehicle_Rotation
   UV_Rotation Bypass
   Mech_Rotation
   Uvs_Offset
   Offset-after rotation
   Uvs_TileRate
  ) {materialFunction=&/MF_Projection_Side_Normal}
  n28 = LinearInterpolate(Alpha=NamedRerouteUsage()) {constA=0.5, constB=1.5}
  NamedRerouteDeclaration(None=LinearInterpolate(
    A=VectorParameter() {group=Wrap_BSequins, parameterName=BS_Falloff_Color_01, defaultValue=<rgba 0.8842 0.6282 1 1>}.RGB
    B=VectorParameter() {group=Wrap_BSequins, parameterName=BS_Falloff_Color_02, defaultValue=<rgba 0.6663 0.575 1 1>}.RGB
    Alpha=Multiply(A=@n28, B=@n5)
   )) {name=SequinColor, nodeColor=<rgba 0.8984 0 1 1>}
  NamedRerouteDeclaration(None=MaterialFunctionCall(Float4=@n26) {materialFunction=&BreakOutFloat4Components}.A) {name=IsCharacter, nodeColor=<rgba 0.8858 1 0.0152 1>}
  n29 = TextureCoordinate()
  n30 = ScalarParameter() {defaultValue=32, group=Wrap_BSequins, parameterName=BS_TileRate}
  NamedRerouteDeclaration(None=LinearInterpolate(
    A=Multiply(A=@n29, B=@n30)
    B=ScalarParameter() {defaultValue=32, group=Wrap_BSequins, parameterName=BS_TileRate_Character}
    Alpha=NamedRerouteUsage()
   )) {name=SequinUVs, nodeColor=<rgba 0 1 0.0859 1>}
  NamedRerouteDeclaration(None=TextureSample(
    UVs=NamedRerouteUsage()
    Tex=TextureObjectParameter() {group=Wrap_BSequins, parameterName=BS_Mask, texture=&T_BlacksRGB}
   ).R) {name=SequinOffset, nodeColor=<rgba 1 0 0.6562 1>}
 ]}
