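# Material function for the BadBear cosmetic wrap, layered onto an input material through primary and secondary masks behind the Wrap_BadBear static switch.
# Side projection and a safe-zone virtual plane supply UVs, with offset, rotation and scale chosen per item type such as weapon, vehicle or backpack.
# The look combines fake subsurface color, curve-atlas noise glow and sine-warped, pulsing emissive lines, plus a separate accent material.
# Time-of-day scales emissive, and a timestamp parameter drives a reactive fade.
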MaterialFunction /Game/Weapons/WeaponSkins/Wraps/MaterialFunctions/MF_Wrap_BadB
graph {nodes [
  n1 = ComponentMask(None=NamedRerouteUsage()) {g=true}
  n2 = Time()
  n3 = ScalarParameter() {defaultValue=0.75, group=Wrap_BadBear_Lines, parameterName=BadBear_PulseSpeed, sortPriority=100}
  n4 = Multiply(A=@n2, B=@n3)
  n5 = Power(Base=OneMinus(None=Frac(None=@n4))) {constExponent=3}
  Add(
   A=Multiply(A=@n1, B=ScalarParameter() {defaultValue=0.02, group=Wrap_BadBear_Lines, parameterName=BadBear_PulseNoiseAmount, sortPriority=100}) {constB=0.04}
   B=@n5
  )
  n6 = Add(B=Multiply(A=@n2, B=@n3)) {constA=0.05}
  n7 = Power(Base=OneMinus(None=Frac(None=@n6))) {constExponent=3}
  Add(
   A=Multiply(A=@n1, B=ScalarParameter() {defaultValue=0.01, group=Wrap_BadBear_Lines, parameterName=BadBear_PulseNoiseAmount2, sortPriority=100}) {constB=0.04}
   B=@n7
  )
  ComponentMask() {g=true, r=true}
  n8 = NamedRerouteUsage()
  ComponentMask(None=@n8) {b=true, r=true}
  n9 = MaterialFunctionCall!(
   Camera _pos
   Axis
   Scale
   Scale_Multplyer
   Object-SafeZone-Virtual-Pivot=NamedRerouteUsage()
   Object-SafeZone-Virtual-CH-Switch=NamedRerouteUsage()
   SinglePlane Normal Axis
   SinglePlane X Axis
   PlaneCenter_distortion
  ) {materialFunction=&/MF_SafeZone_VirtualPlane}
  ComponentMask(None=@n9.Uvs) {r=true}
  Constant3Vector()
  Constant() {r=0.5}
  Constant() {r=0.5}
  n10 = TextureSample() {samplerType=SAMPLERTYPE_Masks, texture=/Game/Weapons/FORT_Rifles/Textures/T_Scar_CM}
  n11 = FunctionInput(Preview=@n10.R) {bUsePreviewValueAsDefault=true, inputName=Mask Primary, inputType=FunctionInput_Scalar, sortPriority=24, previewValue=<1 0 0 1>}
  n12 = FunctionInput(Preview=@n10.B) {bUsePreviewValueAsDefault=true, inputName=Mask Secondary, inputType=FunctionInput_Scalar, sortPriority=24, previewValue=<1 0 0 1>}
  n13 = BlendMaterialAttributes(A=NamedRerouteUsage(), B=NamedRerouteUsage(), Alpha=Saturate(None=Add(A=@n11, B=@n12))) {vertexAttributeBlendType=UseA}
  n14 = NamedRerouteUsage()
  n15 = NamedRerouteDeclaration(None=@n12) {name=Msk_Accent, nodeColor=<rgba 0 1 0.7656 1>}
  n16 = TextureSample() {samplerType=SAMPLERTYPE_Normal, texture=/Game/Weapons/FORT_Rifles/T_SCAR_N}
  n17 = Constant3Vector() {constant=<rgba 1 0 0 0>}
  n18 = GetMaterialAttributes(None=FunctionInput(Preview=SetMaterialAttributes!(MaterialAttributes, Normal=@n16.RGB, BaseColor=@n17) {attributeSetTypes [#1, #2]}) {inputName=Material, inputType=FunctionInput_MaterialAttributes}) {attributeGetTypes [
    #2
    #3
    #1
    #4
   ]}
  FunctionOutput(None=StaticSwitchParameter(
    True=BlendMaterialAttributes(A=@n13, B=@n14, Alpha=@n15) {vertexAttributeBlendType=UseA}
    False=NamedRerouteDeclaration(None=@n18.MaterialAttributes) {name=Root_Material, nodeColor=<rgba 0.7812 1 0 1>}
   ) {group=Customization_Wrap, parameterName=Wrap_BadBear})
  n19 = MaterialFunctionCall(Float4=NamedRerouteUsage()) {materialFunction=&BreakOutFloat4Components}
  n20 = VectorParameter() {group=Wrap_BadBear_InnerGlow, parameterName=BadBear_Inner_GlowNoise_WarpSpeed, defaultValue=<rgba 0.3 0.1 0 1>}
  n21 = MaterialFunctionCall!(
   For_Safe_Zones
   Normal_Map=NamedRerouteUsage()
   Caculated_Bounds=NamedRerouteUsage()
   Pre_Skinned_Normals=NamedRerouteUsage()
   Is Vehicle=@n19.R
   Is VehicleSide=@n19.R
   Is Mech=@n19.R
   Vehicle_Rotation
   UV_Rotation Bypass
   Mech_Rotation
   Uvs_Offset=Multiply(A=Time(), B=@n20.RGB)
   Offset-after rotation
   Uvs_TileRate=ScalarParameter() {defaultValue=2.5, group=Wrap_BadBear_InnerGlow, parameterName=BadBear_Inner_GlowNoise_WarpTileRate}
  ) {materialFunction=&/MF_Projection_Side_Normal}
  n22 = TextureSample(UVs=@n21.UVs, Tex=NamedRerouteUsage())
  n23 = CurveAtlasRowParameter(CurveTime=@n22.A) {
   atlas=/Game/Weapons/WeaponSkins/Wraps/Textures/CruveAtlas_Wraps_01
   curve=/Game/Weapons/WeaponSkins/Wraps/Textures/0_ColorCurves/CC_BadB_Msk
   defaultValue=93
   group=Wrap_BadBear
   parameterName=BadBear_CC_Msking
   sortPriority=0
  }
  n24 = SphereMask(
   A=@n9.Uvs
   B=Constant() {r=0.5}
   Radius=ScalarParameter() {defaultValue=0.6, group=Wrap_BadBear_InnerGlow, parameterName=BadBear_Inner_Glow_Size}
   Hardness=ScalarParameter() {group=Wrap_BadBear_InnerGlow, parameterName=BadBear_Inner_Glow_Hardness}
  )
  n25 = OneMinus(None=@n24)
  n26 = LinearInterpolate(
   A=MaterialFunctionCall!(
    Index=ScalarParameter() {defaultValue=93, group=Wrap_BadBear, parameterName="BadBear_CC_Msking  Index", sortPriority=0}
    CCinput
    CCinput_Alpha=@n23.A
    UV=@n22.A
   ) {materialFunction=&/MF_Global_CC_Tex}.Alpha
   Alpha=@n25
  ) {constA=1}
  n27 = NamedRerouteUsage()
  n28 = Add(A=@n25, B=@n27)
  n29 = ScalarParameter() {defaultValue=93, group=Wrap_BadBear, parameterName="BadBear_CC_Msking  Index", sortPriority=0}
  n30 = ScalarParameter() {defaultValue=0.12, group=Wrap_BadBear_InnerGlow, parameterName=BadBear_Inner_GlowNoise_WarpStrength}
  n31 = NamedRerouteUsage()
  n32 = ComponentMask(None=Add(A=Multiply(A=@n22.A, B=@n30), B=Multiply(A=@n31) {constB=0.76})) {r=true}
  n33 = CurveAtlasRowParameter(CurveTime=@n32) {
   atlas=&CruveAtlas_Wraps_01
   curve=&CC_BadB_Msk
   defaultValue=93
   group=Wrap_BadBear
   parameterName=BadBear_CC_Msking
   sortPriority=0
  }
  MaterialExpressionMaterialFunctionCall_0 = MaterialFunctionCall(Index=@n29, CCinput=@n33, UV=@n32) {materialFunction=&MF_Global_CC_Tex}
  n34 = Convert(None=@MaterialExpressionMaterialFunctionCall_0.3V) {
   nodeName=Break Float3
   convertInputs [{
     type=Vector3
     expressionInput {
      expression=&self:MaterialExpressionMaterialFunctionCall_0
      inputName=None
      mask=0
      maskA=0
      maskB=0
      maskG=0
      maskR=0
      outputIndex=0
     }
    }]
   convertOutputs [{type=Scalar}, {type=Scalar}, {type=Scalar}]
  }
  LinearInterpolate(
   A=LinearInterpolate(
    A=VectorParameter() {group=Wrap_BadBear, parameterName=BadBear_Inner_ColorA, defaultValue=<rgba 0 0.0292 0.1 1>}.RGB
    B=VectorParameter() {group=Wrap_BadBear, parameterName=BadBear_Inner_ColorB, defaultValue=<rgba 0.0782 0.3712 0.5841 1>}.RGB
    Alpha=Multiply(A=@n26, B=@n28) {constA=1}
   )
   B=LinearInterpolate(
    A=VectorParameter() {group=Wrap_BadBear, parameterName=BadBear_Inner_ColorC, defaultValue=<rgba 0.0943 0.2656 0.065 1>}.RGB
    B=VectorParameter() {group=Wrap_BadBear, parameterName=BadBear_Inner_ColorD, defaultValue=<rgba 0.6719 0.6429 0.0798 1>}.RGB
    Alpha=@n34
   )
   Alpha=@n34
  )
  n35 = MaterialFunctionCall(Float4=NamedRerouteUsage()) {materialFunction=&BreakOutFloat4Components}
  n36 = MaterialFunctionCall(
   Normal_Map=NamedRerouteUsage()
   Caculated_Bounds=NamedRerouteUsage()
   Pre_Skinned_Normals=NamedRerouteUsage()
   Is Vehicle=@n35.R
   Is VehicleSide=@n35.R
   Is Mech=@n35.R
   UV_Rotation Bypass=NamedRerouteUsage()
   Uvs_Offset=NamedRerouteUsage()
   Uvs_TileRate=NamedRerouteUsage()
  ) {materialFunction=&MF_Projection_Side_Normal}
  LinearInterpolate(A=NamedRerouteUsage(), B=@n36.Raw_Normal, Alpha=NamedRerouteUsage())
  n37 = VectorParameter() {group=Wrap_BadBear_Subsurface, parameterName=BadBear_InnerColor, defaultValue=<rgba 0 0.0821 0.23 1>}
  n38 = VectorParameter() {group=Wrap_BadBear_Subsurface, parameterName=BadBear_OuterColorA, defaultValue=<rgba 0.0125 0.724 0.6663 1>}
  n39 = Subtract(
   A=ComponentMask(None=NamedRerouteUsage()) {b=true}
   B=ScalarParameter() {defaultValue=0.1, group=Wrap_BadBear_Subsurface, parameterName=BadBear_OuterColor_TransitionStart}
  ) {constB=125}
  n40 = ScalarParameter() {defaultValue=0.5, group=Wrap_BadBear_Subsurface, parameterName=BadBear_OuterColor_TransitionScale}
  n41 = LinearInterpolate(
   A=@n38.RGB
   B=VectorParameter() {group=Wrap_BadBear_Subsurface, parameterName=BadBear_OuterColorB, defaultValue=<rgba 0 0.5903 0.75 0>}.RGB
   Alpha=Saturate(None=Multiply(A=@n39, B=@n40) {constB=0.015})
  )
  n42 = NamedRerouteUsage()
  n43 = Multiply(
   A=Add(A=@n42) {constB=-0.5}
   B=ScalarParameter() {defaultValue=0.2, group=Wrap_BadBear_Subsurface, parameterName=BadBear_Curvature_Wrap_Amount}
  )
  n44 = OneMinus(None=SphereMask(
    A=NamedRerouteUsage()
    B=Constant() {r=0.5}
    Radius=NamedRerouteUsage()
    Hardness=ScalarParameter() {defaultValue=0.1, group=Wrap_BadBear_Subsurface, parameterName=BadBear_SubsFalloff}
   ))
  n45 = Multiply(
   A=Fresnel(ExponentIn=ScalarParameter() {defaultValue=8, group=Wrap_BadBear_Subsurface, parameterName=BadBear_HighlightSpecular_Scale}) {baseReflectFraction=0}
   B=ScalarParameter() {defaultValue=0.5, group=Wrap_BadBear_Subsurface, parameterName=BadBear_Fresnel Strenght}
  )
  n46 = Add(A=@n43, B=Add(A=Power(Base=@n44), B=@n45))
  n47 = VectorParameter() {group=Wrap_BadBear_Subsurface, parameterName=BadBear_OuterColorC, defaultValue=<rgba 0.75 0.0062 0 0>}
  n48 = NamedRerouteUsage()
  n49 = ComponentMask(None=VectorParameter() {group=Wrap_BadBear_Subsurface, parameterName=BadBear_SubsPos_B, defaultValue=<rgba 0.255 0.5 0 1>}.RGB) {g=true, r=true}
  n50 = NamedRerouteUsage()
  n51 = ScalarParameter() {defaultValue=0.1, group=Wrap_BadBear_Subsurface, parameterName=BadBear_SubsFalloff_B}
  n52 = Power(Base=OneMinus(None=SphereMask(A=@n48, B=@n49, Radius=@n50, Hardness=@n51)))
  n53 = Add(A=Add(A=@n52, B=@n45), B=@n43)
  n54 = VectorParameter() {group=Wrap_BadBear_Subsurface, parameterName=BadBear_InnerScatteringColor, defaultValue=<rgba 0.0877 0.0586 0.2031 1>}
  n55 = CollectionParameter() {
   collection=&FortniteMaterialParameters
   parameterName=SunlightDirectionalVector
  }
  n56 = ComponentMask(None=@n55) {b=true, g=true, r=true}
  n57 = VertexNormalWS()
  n58 = Multiply(A=DotProduct(A=@n56, B=@n57)) {constB=2}
  n59 = Normalize(VectorInput=Subtract(A=Multiply(A=@n58, B=@n57), B=@n56))
  n60 = Multiply(A=MaterialFunctionCall() {materialFunction=&CameraDirectionVector}.Result) {constB=-1}
  n61 = Power(Base=Max(A=DotProduct(A=@n59, B=@n60)) {constB=0}, Exp=Constant() {r=5})
  n62 = LinearInterpolate(
   A=LinearInterpolate(A=LinearInterpolate(A=@n37.RGB, B=@n41, Alpha=@n46), B=@n47.RGB, Alpha=Multiply(A=@n47.A, B=@n53))
   B=@n54.RGB
   Alpha=Saturate(None=Multiply(A=@n61) {constB=0.4})
  )
  Multiply(
   A=MaterialFunctionCall(Value_Morning=Constant() {r=0.66}, Value_Day=Constant() {r=1}, Value_Evening=Constant() {r=0.66}, Value_Night=Constant() {r=0.5}) {materialFunction=&Scalar_Time_of_Day}.Result
   B=@n62
  )
  Multiply(A=@n52)
  Multiply(A=MaterialFunctionCall(Value_Morning=Constant() {r=0.5}, Value_Day=Constant() {r=1}, Value_Evening=Constant() {r=0.5}, Value_Night=Constant() {r=0.25}) {materialFunction=&Scalar_Time_of_Day}.Result)
  Multiply(A=OneMinus(None=@n42))
  n63 = ScalarParameter() {defaultValue=5, group=Wrap_BadBear_Lines, parameterName=BadBear_Line_Fresnel_Scale}
  n64 = ScalarParameter() {defaultValue=0.5, group=Wrap_BadBear_Lines, parameterName=BadBear_Line_Fresnel_Strenght}
  Multiply(A=OneMinus(None=Fresnel(ExponentIn=@n63) {baseReflectFraction=0}), B=@n64)
  n65 = NamedRerouteUsage()
  Multiply(A=@n65, B=@n64)
  Multiply(A=@n44)
  n66 = Constant2Vector() {g=0.5}
  n67 = SphereMask(A=@n8, B=@n66) {attenuationRadius=1.1, hardnessPercent=0}
  Multiply(A=@n67)
  n68 = VectorParameter() {group=Wrap_BadBear_Accent, parameterName=BadBear_Accent_BaseColor, defaultValue=<rgba 0.025 0.0236 0.0222 1>}
  n69 = SetMaterialAttributes!(
   MaterialAttributes=NamedRerouteUsage()
   BaseColor=@n68.RGB
   Roughness=ScalarParameter() {defaultValue=0.4, group=Wrap_BadBear_Accent, parameterName=BadBear_Accent_Roughness}
   Specular=ScalarParameter() {defaultValue=0.5, group=Wrap_BadBear_Accent, parameterName=BadBear_Accent_Spec}
   Metallic=ScalarParameter() {group=Wrap_BadBear_Accent, parameterName=BadBear_Accent_Metalic}
   Normal
   EmissiveColor
  ) {attributeSetTypes [
    #2
    #3
    #5
    #4
    #1
    #6
   ]}
  NamedRerouteDeclaration(None=@n69) {name=Mat_Accent, nodeColor=<rgba 0 0.3906 1 1>}
  n70 = FunctionInput(Preview=@n10.G) {bUsePreviewValueAsDefault=true, inputName=Mask Crevise, inputType=FunctionInput_Scalar, sortPriority=24, previewValue=<1 0 0 1>}
  NamedRerouteDeclaration(None=@n70) {name=Msk_Curvature, nodeColor=<rgba 0 1 0.7422 1>}
  n71 = FunctionInput(Preview=Constant4Vector() {constant=<rgba 0 0 5 -1>}.RGBA) {bUsePreviewValueAsDefault=true, inputName=Object-SafeZone-Virtual-Pivot, inputType=FunctionInput_Vector4, sortPriority=105}
  NamedRerouteDeclaration(None=@n71) {name=SafeZone VP, nodeColor=<rgba 1 0 0.8203 1>}
  n72 = TextureObjectParameter() {group=Wrap_BadBear, parameterName=BadBear__Tex_Msk, sortPriority=5, texture=&T_BlacksRGB}
  NamedRerouteDeclaration(None=@n72) {name=Tex_Msk, nodeColor=<rgba 0.7344 0 1 1>}
  n73 = FunctionInput(Preview=Constant()) {bUsePreviewValueAsDefault=true, inputName=Object-SafeZone-VP-Character-PreView-Switch, inputType=FunctionInput_Scalar, sortPriority=105}
  NamedRerouteDeclaration(None=@n73) {name=SafeZone Character-Switch, nodeColor=<rgba 1 0 0.8203 1>}
  n74 = TextureObjectParameter() {group=Wrap_BadBear, parameterName=BadBear__Tex_Ms2, sortPriority=5, texture=&T_BlacksRGB}
  NamedRerouteDeclaration(None=@n74) {name=Tex_Msk2, nodeColor=<rgba 0.7344 0 1 1>}
  n75 = TextureObjectParameter() {
   group=Wrap_BadBear
   parameterName=BadBear__Tex_N
   samplerType=SAMPLERTYPE_Normal
   sortPriority=5
   texture=&FlatNormalBlue
  }
  NamedRerouteDeclaration(None=@n75) {name=Tex_N, nodeColor=<rgba 0.7344 0 1 1>}
  NamedRerouteDeclaration(None=@n36.UVs) {name=UVs, nodeColor=<rgba 0.5703 1 0 1>}
  n76 = TextureSample(UVs=NamedRerouteUsage(), Tex=NamedRerouteUsage()) {samplerType=SAMPLERTYPE_Normal}
  NamedRerouteDeclaration(None=@n76.B) {name=Base_Roughness, nodeColor=<rgba 1 0.7969 0 1>}
  n77 = TextureSample(UVs=@n36.UVs, Tex=NamedRerouteUsage()) {samplerType=SAMPLERTYPE_Normal}
  n78 = Multiply(A=@n77.RGB, B=@n77.RGB)
  n79 = ScalarParameter() {defaultValue=0.8, group=Wrap_BadBear, parameterName=BadBear_Base_Spec, sortPriority=10}
  n80 = NamedRerouteUsage()
  n81 = ScalarParameter() {defaultValue=0.55, group=Wrap_BadBear, parameterName=BadBear_Base_Metalic, sortPriority=10}
  n82 = VectorParameter() {group=Wrap_BadBear_Lines, parameterName=BadBear_Line_Glow, defaultValue=<rgba 1 0.7748 0 1>}
  n83 = Saturate(None=OneMinus(None=@n65))
  n84 = NamedRerouteUsage()
  n85 = Multiply(A=OneMinus(None=NamedRerouteUsage()), B=Add(A=Multiply(A=@n82.RGB, B=@n65), B=Multiply(A=@n83, B=@n84)))
  n86 = NamedRerouteUsage()
  n87 = LinearInterpolate(
   A=@n85
   B=Multiply(A=@n85, B=ScalarParameter() {defaultValue=2.5, group=Wrap_BadBear, parameterName=BadBear_Inner_Curvatur})
   Alpha=Clamp(None=Add(A=@n86) {constB=-0.5}) {minDefault=-0.1}
  )
  n88 = MaterialFunctionCall(Value_Morning=Constant() {r=0.5}, Value_Day=Constant() {r=1}, Value_Evening=Constant() {r=0.5}, Value_Night=Constant() {r=0.25}) {materialFunction=&Scalar_Time_of_Day}
  n89 = SetMaterialAttributes(
   MaterialAttributes=NamedRerouteUsage()
   BaseColor=@n78
   Roughness=LinearInterpolate(
    A=ScalarParameter() {defaultValue=0.2, group=Wrap_BadBear, parameterName=BadBear_Slurp_Roughness, sortPriority=10}
    B=NamedRerouteUsage()
    Alpha=@n77.A
   )
   Specular=LinearInterpolate(
    A=ScalarParameter() {defaultValue=0.5, group=Wrap_BadBear, parameterName=BadBear_Slurp_Spec, sortPriority=10}
    B=Multiply(A=@n79, B=@n80)
    Alpha=@n77.A
   )
   Metallic=LinearInterpolate(A=ScalarParameter() {group=Wrap_BadBear, parameterName=BadBear_Slurp_Metalic, sortPriority=10}, B=Multiply(A=@n81, B=@n80), Alpha=@n77.A)
   Normal=Add(A=@n36.Raw_Normal, B=NamedRerouteUsage()) {constB=0}
   EmissiveColor=Clamp(None=Multiply(A=@n87, B=@n88.Result), Max=ScalarParameter() {defaultValue=10, group=Wrap_BadBear, parameterName=BadBear_Max_Emissive}) {maxDefault=10}
  ) {attributeSetTypes [
    #2
    #3
    #5
    #4
    #1
    #6
   ]}
  NamedRerouteDeclaration(None=@n89) {name=Mat_Base, nodeColor=<rgba 0 0.3906 1 1>}
  n90 = Add(A=@n76.G) {constB=0.5}
  NamedRerouteDeclaration(None=Power(Base=@n90)) {name=Base_Cur_low, nodeColor=<rgba 1 0.7969 0 1>}
  n91 = VectorParameter() {group=Wrap_BadBear_Uvs, parameterName=BadBear_Offset_Vehicle, defaultValue=<rgba 0 0 0 1>}
  n92 = VectorParameter() {group=Wrap_BadBear_Uvs, parameterName=BadBear_Offset_VehicleSide, defaultValue=<rgba 0 0 0 1>}
  n93 = MaterialFunctionCall(Float4=NamedRerouteUsage()) {materialFunction=&BreakOutFloat4Components}
  n94 = LinearInterpolate(
   A=LinearInterpolate(
    A=VectorParameter() {group=Wrap_BadBear_Uvs, parameterName=BadBear_Offset_Weapon, defaultValue=<rgba 0 0.078 0 1>}.RGB
    B=LinearInterpolate(A=@n91.RGB, B=@n92.RGB, Alpha=@n93.G)
    Alpha=@n93.R
   )
   B=LinearInterpolate(
    A=VectorParameter() {group=Wrap_BadBear_Uvs, parameterName=BadBear_Offset_Mech, defaultValue=<rgba 0 0 0 1>}.RGB
    B=VectorParameter() {group=Wrap_BadBear_Uvs, parameterName=BadBear_Offset_Character, defaultValue=<rgba 0 0 0 1>}.RGB
    Alpha=@n93.A
   )
   Alpha=@n93.B
  )
  n95 = VectorParameter() {group=Wrap_BadBear_Uvs, parameterName=BadBear_Offset_Backpack, defaultValue=<rgba 0 0 0 1>}
  n96 = MaterialFunctionCall(Float2=NamedRerouteUsage()) {materialFunction=&BreakOutFloat2Components}
  NamedRerouteDeclaration(None=LinearInterpolate(
    A=LinearInterpolate(A=@n94, B=@n95.RGB, Alpha=@n96.R)
    B=VectorParameter() {group=Wrap_BadBear_Uvs, parameterName=BadBear_Offset_Pickaxe, defaultValue=<rgba 0 0 0 1>}.RGB
    Alpha=@n96.G
   )) {name=UVs_Offset_Side, nodeColor=<rgba 0 0.625 1 1>}
  NamedRerouteDeclaration(None=@n76.R) {name=Base_AO, nodeColor=<rgba 1 0.7969 0 1>}
  n97 = TextureObjectParameter() {
   group=Wrap_BadBear
   parameterName=BadBear__Tex_N2
   samplerType=SAMPLERTYPE_Normal
   sortPriority=5
   texture=&FlatNormalBlue
  }
  NamedRerouteDeclaration(None=@n97) {name=Tex_N2, nodeColor=<rgba 0.7344 0 1 1>}
  NamedRerouteDeclaration(None=@n21.Raw_Normal) {name=Warp_Noise_Normal, nodeColor=<rgba 0 1 0.0156 1>}
  NamedRerouteDeclaration(None=@n77.A) {name=BC_Mask, nodeColor=<rgba 0.7344 1 0 1>}
  n98 = ComponentMask(None=@n31) {r=true}
  n99 = CurveAtlasRowParameter(CurveTime=@n98) {
   atlas=&CruveAtlas_Wraps_01
   curve=&CC_BadB_Msk
   defaultValue=93
   group=Wrap_BadBear
   parameterName=BadBear_CC_Msking
   sortPriority=0
  }
  MaterialExpressionMaterialFunctionCall_10 = MaterialFunctionCall(Index=@n29, CCinput=@n99, UV=@n98) {materialFunction=&MF_Global_CC_Tex}
  n100 = Convert(None=@MaterialExpressionMaterialFunctionCall_10.3V) {
   nodeName=Break Float3
   convertInputs [{
     type=Vector3
     expressionInput {
      expression=&self:MaterialExpressionMaterialFunctionCall_10
      inputName=None
      mask=0
      maskA=0
      maskB=0
      maskG=0
      maskR=0
      outputIndex=0
     }
    }]
   convertOutputs [{type=Scalar}, {type=Scalar}, {type=Scalar}]
  }
  NamedRerouteDeclaration(None=@n100) {name=Frnt_Msking_A, nodeColor=<rgba 1 0.0469 0 1>}
  NamedRerouteDeclaration(None=@n78) {name=BC_color, nodeColor=<rgba 0.7344 1 0 1>}
  n101 = NamedRerouteUsage()
  n102 = NamedRerouteUsage()
  NamedRerouteDeclaration(None=Power(Base=TextureSample(UVs=@n101, Tex=@n102) {samplerType=SAMPLERTYPE_Normal}.A) {constExponent=2.5}) {name=Masking Lines, nodeColor=<rgba 0.875 1 0 1>}
  n103 = SphereMask(A=@n67, B=@n5) {attenuationRadius=0.015, hardnessPercent=1}
  n104 = Multiply(A=@n103, B=ScalarParameter() {defaultValue=0.01, group=Wrap_BadBear_Lines, parameterName=BadBear_PulseNoise_UnderlyingEms, sortPriority=100}) {constB=0.01}
  n105 = ScalarParameter() {defaultValue=1, group=Wrap_BadBear_Lines, parameterName=BadBear_Line_Brightness_Reactive, sortPriority=50}
  n106 = ScalarParameter() {defaultValue=1, group=Wrap_BadBear_Lines, parameterName=BadBear_Line_Brightness, sortPriority=50}
  n107 = NamedRerouteUsage()
  n108 = Multiply(A=Time(), B=ScalarParameter() {defaultValue=6, group=Wrap_BadBear_Lines, parameterName=BadBear_Line_Static_Speed, sortPriority=50})
  n109 = Add(A=Sine(None=@n108), B=Sine(None=@n108) {period=0.5}) {constB=0}
  n110 = NamedRerouteUsage()
  n111 = Power(Base=Add(A=OneMinus(None=@n110), B=NamedRerouteUsage())) {constExponent=10}
  n112 = NamedRerouteUsage()
  n113 = Multiply(A=NamedRerouteUsage(), B=NamedRerouteUsage())
  n114 = Multiply(A=Time(), B=ScalarParameter() {defaultValue=-1.0335, group=Wrap_BadBear_Lines, parameterName=BadBear_Line_Speed_A})
  n115 = NamedRerouteUsage()
  n116 = Multiply(
   A=ComponentMask(None=@n115) {r=true}
   B=ScalarParameter() {defaultValue=1.5, group=Wrap_BadBear_Lines, parameterName=BadBear_Line_TileRate_A}
  ) {constB=1.5}
  n117 = ScalarParameter() {defaultValue=0.05, group=Wrap_BadBear_Lines, parameterName=BadBear_Line_Strength_A}
  n118 = NamedRerouteUsage()
  n119 = Multiply(A=Sine(None=Add(A=@n114, B=@n116)), B=LinearInterpolate(A=@n117, B=Multiply(A=@n117, B=@n118), Alpha=NamedRerouteUsage())) {constB=0.05}
  n120 = Add(A=Multiply(A=@n113, B=Multiply(A=@n119) {constB=-1}), B=@n115)
  n121 = NamedRerouteUsage()
  n122 = TextureSample(UVs=Saturate(None=Add(A=@n112, B=@n120)), Tex=@n121) {samplerType=SAMPLERTYPE_Normal}
  n123 = Add(A=@n115, B=Multiply(A=@n113, B=@n119))
  n124 = TextureSample(UVs=Saturate(None=Add(A=@n112, B=@n123)), Tex=@n121) {samplerType=SAMPLERTYPE_Normal}
  n125 = Add(A=Multiply(A=@n122.R, B=OneMinus(None=@n122.B)), B=Multiply(A=@n124.R, B=@n124.R))
  n126 = NamedRerouteUsage()
  n127 = Multiply(A=NamedRerouteUsage(), B=NamedRerouteUsage())
  n128 = Multiply(A=Time(), B=ScalarParameter() {defaultValue=-0.803, group=Wrap_BadBear_Lines, parameterName=BadBear_Line_Speed_B})
  n129 = NamedRerouteUsage()
  n130 = Multiply(
   A=ComponentMask(None=@n129) {r=true}
   B=ScalarParameter() {defaultValue=1.75, group=Wrap_BadBear_Lines, parameterName=BadBear_Line_TileRate_B}
  ) {constB=1.75}
  n131 = ScalarParameter() {defaultValue=0.07, group=Wrap_BadBear_Lines, parameterName=BadBear_Line_Strength_B}
  n132 = NamedRerouteUsage()
  n133 = Multiply(A=Sine(None=Add(A=@n128, B=@n130)), B=LinearInterpolate(A=@n131, B=Multiply(A=@n131, B=@n132), Alpha=NamedRerouteUsage())) {constB=0.07}
  n134 = Add(A=Multiply(A=@n127, B=Multiply(A=@n133) {constB=-1}), B=@n129)
  n135 = NamedRerouteUsage()
  n136 = TextureSample(UVs=Saturate(None=Add(A=@n126, B=@n134)), Tex=@n135) {samplerType=SAMPLERTYPE_Normal}
  n137 = Multiply(A=@n136.R, B=OneMinus(None=@n136.B))
  n138 = Add(A=@n129, B=Multiply(A=@n127, B=@n133))
  n139 = TextureSample(UVs=Saturate(None=Add(A=@n126, B=@n138)), Tex=@n135) {samplerType=SAMPLERTYPE_Normal}
  n140 = Multiply(A=@n139.R, B=@n139.R)
  n141 = Multiply(A=NamedRerouteUsage(), B=NamedRerouteUsage())
  n142 = Multiply(A=Time(), B=ScalarParameter() {defaultValue=-0.9034, group=Wrap_BadBear_Lines, parameterName=BadBear_Line_Speed_C})
  n143 = NamedRerouteUsage()
  n144 = Multiply(A=ComponentMask(None=@n143) {r=true}, B=ScalarParameter() {defaultValue=2, group=Wrap_BadBear_Lines, parameterName=BadBear_Line_TileRate_C}) {constB=2}
  n145 = ScalarParameter() {defaultValue=0.05, group=Wrap_BadBear_Lines, parameterName=BadBear_Line_Strength_C}
  n146 = NamedRerouteUsage()
  n147 = Multiply(A=Sine(None=Add(A=@n142, B=@n144)), B=LinearInterpolate(A=@n145, B=Multiply(A=@n145, B=@n146), Alpha=NamedRerouteUsage())) {constB=0.05}
  n148 = Add(A=Multiply(A=@n141, B=Multiply(A=@n147) {constB=-1}), B=@n143)
  n149 = NamedRerouteUsage()
  n150 = NamedRerouteUsage()
  n151 = TextureSample(UVs=Saturate(None=Add(A=@n148, B=@n149)), Tex=@n150) {samplerType=SAMPLERTYPE_Normal}
  n152 = Multiply(A=@n151.R, B=OneMinus(None=@n151.B))
  n153 = Add(A=@n143, B=Multiply(A=@n141, B=@n147))
  n154 = TextureSample(UVs=Saturate(None=Add(A=@n149, B=@n153)), Tex=@n150) {samplerType=SAMPLERTYPE_Normal}
  n155 = Multiply(A=@n154.R, B=@n154.R)
  n156 = Add(A=Add(A=Multiply(A=@n125) {constB=0.25}, B=Add(A=@n137, B=@n140)), B=Multiply(A=Add(A=@n152, B=@n155)) {constB=0.5})
  n157 = Multiply(A=ScalarParameter() {defaultValue=2, group=Wrap_BadBear_Lines, parameterName=BadBear_Line_Static_Brightness, sortPriority=50}, B=@n156) {constA=5}
  n158 = Add(A=Multiply(A=ConstantBiasScale(None=@n109) {bias=2}, B=Multiply(A=@n111, B=@n157)), B=@n156)
  n159 = Add(
   A=Multiply(A=@n103, B=ScalarParameter() {defaultValue=5, group=Wrap_BadBear_Lines, parameterName=BadBear_PulseNoiseEms, sortPriority=100}) {constB=5}
   B=Multiply(
    A=SphereMask(A=@n67, B=@n7) {attenuationRadius=0.04, hardnessPercent=1}
    B=ScalarParameter() {defaultValue=10, group=Wrap_BadBear_Lines, parameterName=BadBear_PulseNoiseEms2, sortPriority=100}
   )
  )
  n160 = SphereMask(A=@n8, B=@n66) {attenuationRadius=2, hardnessPercent=0}
  n161 = Add(A=Multiply(A=@n158, B=LinearInterpolate(A=@n105, B=@n106, Alpha=@n107)) {constB=0}, B=Multiply(A=@n158, B=Multiply(A=@n159, B=@n160)))
  NamedRerouteDeclaration(None=Multiply(A=NamedRerouteUsage(), B=Add(A=@n104, B=@n161))) {name=Line Glows, nodeColor=<rgba 1 0 0.1641 1>}
  n162 = SetMaterialAttributes!(
   MaterialAttributes=NamedRerouteUsage()
   BaseColor
   Metallic=Constant()
   Specular=@n42
   Roughness=ScalarParameter() {defaultValue=0.15, parameterName=Param_6}
   EmissiveColor
  ) {attributeSetTypes [
    #2
    #4
    #5
    #3
    #6
   ]}
  NamedRerouteDeclaration(None=@n162) {name=testing material, nodeColor=<rgba 0 0.7188 1 1>}
  n163 = NamedRerouteUsage()
  NamedRerouteDeclaration(None=Multiply(
    A=Add(A=@n163) {constB=-0.5}
    B=ScalarParameter() {defaultValue=-0.015, group=Wrap_BadBear_Lines, parameterName=BadBear_Line_Curavature_Warp_Strength}
   )) {name=Warp_Curavature, nodeColor=<rgba 0.3125 1 0 1>}
  n164 = MaterialFunctionCall(Color=Abs(None=NamedRerouteUsage())) {materialFunction=&SplitComponents}
  n165 = MaterialFunctionCall(Float4=NamedRerouteUsage()) {materialFunction=&BreakOutFloat4Components}
  n166 = LinearInterpolate(A=@n164.R, B=LinearInterpolate(A=@n164.B, B=@n164.B, Alpha=@n165.G), Alpha=@n165.R)
  NamedRerouteDeclaration(None=MaterialFunctionCall(
    In=LinearInterpolate(A=@n166, B=@n164.G, Alpha=@n165.B)
    Contrast=ScalarParameter() {defaultValue=1, group=Wrap_BadBear, parameterName=BadBear_Blend_Contrast, sortPriority=0}
   ) {materialFunction=&CheapContrast}.Result) {name=Tri_Mask, nodeColor=<rgba 1 0.4688 0 1>}
  NamedRerouteDeclaration(None=@n100) {name=Frnt_Msking_B, nodeColor=<rgba 1 0.0469 0 1>}
  n167 = ScalarParameter() {group=Wrap_BadBear_Uvs, parameterName=BadBear_Rotation_Vehicle, sortPriority=150}
  n168 = ScalarParameter() {group=Wrap_BadBear_Uvs, parameterName=BadBear_Rotation_Vehicle_Side, sortPriority=150}
  n169 = MaterialFunctionCall(Float4=NamedRerouteUsage()) {materialFunction=&BreakOutFloat4Components}
  n170 = LinearInterpolate(
   A=LinearInterpolate(
    A=ScalarParameter() {group=Wrap_BadBear_Uvs, parameterName=BadBear_Rotation_Weapon, sortPriority=150}
    B=LinearInterpolate(A=@n167, B=@n168, Alpha=@n169.G)
    Alpha=@n169.R
   )
   B=LinearInterpolate(
    A=ScalarParameter() {group=Wrap_BadBear_Uvs, parameterName=BadBear_Rotation_Mech, sortPriority=150}
    B=ScalarParameter() {group=Wrap_BadBear_Uvs, parameterName=BadBear_Rotation_Character, sortPriority=150}
    Alpha=@n169.A
   )
   Alpha=@n169.B
  )
  n171 = ScalarParameter() {group=Wrap_BadBear_Uvs, parameterName=BadBear_Rotation_Backpack, sortPriority=150}
  n172 = MaterialFunctionCall(Float2=NamedRerouteUsage()) {materialFunction=&BreakOutFloat2Components}
  NamedRerouteDeclaration(None=LinearInterpolate(
    A=LinearInterpolate(A=@n170, B=@n171, Alpha=@n172.R)
    B=ScalarParameter() {group=Wrap_BadBear_Uvs, parameterName=BadBear_Rotation_Pickaxe, sortPriority=150}
    Alpha=@n172.G
   )) {name=UVs_Rotation, nodeColor=<rgba 0 0.7188 1 1>}
  n173 = ScalarParameter() {defaultValue=1, group=Wrap_BadBear_Uvs, parameterName=BadBear_Scale_Vehicle, sortPriority=150}
  n174 = ScalarParameter() {defaultValue=1, group=Wrap_BadBear_Uvs, parameterName=BadBear_Scale_Vehicle_Side, sortPriority=150}
  n175 = MaterialFunctionCall(Float4=NamedRerouteUsage()) {materialFunction=&BreakOutFloat4Components}
  n176 = LinearInterpolate(
   A=LinearInterpolate(
    A=ScalarParameter() {defaultValue=1, group=Wrap_BadBear_Uvs, parameterName=BadBear_Scale_Weapon, sortPriority=150}
    B=LinearInterpolate(A=@n173, B=@n174, Alpha=@n175.G)
    Alpha=@n175.R
   )
   B=LinearInterpolate(
    A=ScalarParameter() {defaultValue=1, group=Wrap_BadBear_Uvs, parameterName=BadBear_Scale_Mech, sortPriority=150}
    B=ScalarParameter() {defaultValue=1, group=Wrap_BadBear_Uvs, parameterName=BadBear_Scale_Character, sortPriority=150}
    Alpha=@n175.A
   )
   Alpha=@n175.B
  )
  n177 = ScalarParameter() {defaultValue=1, group=Wrap_BadBear_Uvs, parameterName=BadBear_Scale_Backpack, sortPriority=150}
  n178 = MaterialFunctionCall(Float2=NamedRerouteUsage()) {materialFunction=&BreakOutFloat2Components}
  NamedRerouteDeclaration(None=LinearInterpolate(
    A=LinearInterpolate(A=@n176, B=@n177, Alpha=@n178.R)
    B=ScalarParameter() {defaultValue=1, group=Wrap_BadBear_Uvs, parameterName=BadBear_Scale_Pickaxe, sortPriority=150}
    Alpha=@n178.G
   )) {name=UVs_Scale, nodeColor=<rgba 0 0.7188 1 1>}
  NamedRerouteDeclaration(None=@n36.UVs No Offsets/Tiling) {name=UVs_No, nodeColor=<rgba 0.5703 1 0 1>}
  n179 = VertexInterpolator(VS=PreSkinnedPosition())
  n180 = PreSkinnedLocalBounds()
  NamedRerouteDeclaration(None=Divide(A=Subtract(A=@n179.PS, B=@n180.Min), B=Subtract(A=@n180.Max, B=@n180.Max))) {name=CaculatedBounds_B, nodeColor=<rgba 0 0.3438 1 1>}
  NamedRerouteDeclaration(None=@n62) {name=Base_Emissive, nodeColor=<rgba 0 0.9766 1 1>}
  NamedRerouteDeclaration(None=@n42) {name=Base_Emissive_Spec, nodeColor=<rgba 0 0.9766 1 1>}
  n181 = ScalarParameter() {defaultValue=1, group=Wrap_BadBear_Subsurface, parameterName=BadBear_SubsSizeA_Vehicle, sortPriority=150}
  n182 = ScalarParameter() {defaultValue=1, group=Wrap_BadBear_Subsurface, parameterName=BadBear_SubsSizeA_Vehicle_Side, sortPriority=150}
  n183 = MaterialFunctionCall(Float4=NamedRerouteUsage()) {materialFunction=&BreakOutFloat4Components}
  n184 = LinearInterpolate(
   A=LinearInterpolate(
    A=ScalarParameter() {defaultValue=1, group=Wrap_BadBear_Subsurface, parameterName=BadBear_SubsSizeA_Weapon, sortPriority=150}
    B=LinearInterpolate(A=@n181, B=@n182, Alpha=@n183.G)
    Alpha=@n183.R
   )
   B=LinearInterpolate(
    A=ScalarParameter() {defaultValue=1, group=Wrap_BadBear_Subsurface, parameterName=BadBear_SubsSizeA_Mech, sortPriority=150}
    B=ScalarParameter() {defaultValue=1, group=Wrap_BadBear_Subsurface, parameterName=BadBear_SubsSizeA_Character, sortPriority=150}
    Alpha=@n183.A
   )
   Alpha=@n183.B
  )
  n185 = ScalarParameter() {defaultValue=1, group=Wrap_BadBear_Subsurface, parameterName=BadBear_SubsSizeA_Backpack, sortPriority=150}
  n186 = MaterialFunctionCall(Float2=NamedRerouteUsage()) {materialFunction=&BreakOutFloat2Components}
  NamedRerouteDeclaration(None=LinearInterpolate(
    A=LinearInterpolate(A=@n184, B=@n185, Alpha=@n186.R)
    B=ScalarParameter() {defaultValue=1, group=Wrap_BadBear_Subsurface, parameterName=BadBear_SubsSizeA_Pickaxe, sortPriority=150}
    Alpha=@n186.G
   )) {name=SubsScale_A, nodeColor=<rgba 0 0.7188 1 1>}
  n187 = ScalarParameter() {defaultValue=1, group=Wrap_BadBear_Subsurface, parameterName=BadBear_SubsSizeB_Vehicle, sortPriority=150}
  n188 = ScalarParameter() {defaultValue=1, group=Wrap_BadBear_Subsurface, parameterName=BadBear_SubsSizeB_Vehicle_Side, sortPriority=150}
  n189 = MaterialFunctionCall(Float4=NamedRerouteUsage()) {materialFunction=&BreakOutFloat4Components}
  n190 = LinearInterpolate(
   A=LinearInterpolate(
    A=ScalarParameter() {defaultValue=0.5, group=Wrap_BadBear_Subsurface, parameterName=BadBear_SubsSizeB_Weapon, sortPriority=150}
    B=LinearInterpolate(A=@n187, B=@n188, Alpha=@n189.G)
    Alpha=@n189.R
   )
   B=LinearInterpolate(
    A=ScalarParameter() {defaultValue=1, group=Wrap_BadBear_Subsurface, parameterName=BadBear_SubsSizeB_Mech, sortPriority=150}
    B=ScalarParameter() {defaultValue=1, group=Wrap_BadBear_Subsurface, parameterName=BadBear_SubsSizeB_Character, sortPriority=150}
    Alpha=@n189.A
   )
   Alpha=@n189.B
  )
  n191 = ScalarParameter() {defaultValue=1, group=Wrap_BadBear_Subsurface, parameterName=BadBear_SubsSizeB_Backpack, sortPriority=150}
  n192 = MaterialFunctionCall(Float2=NamedRerouteUsage()) {materialFunction=&BreakOutFloat2Components}
  NamedRerouteDeclaration(None=LinearInterpolate(
    A=LinearInterpolate(A=@n190, B=@n191, Alpha=@n192.R)
    B=ScalarParameter() {defaultValue=1, group=Wrap_BadBear_Subsurface, parameterName=BadBear_SubsSizeB_Pickaxe, sortPriority=150}
    Alpha=@n192.G
   )) {name=SubsScale_B, nodeColor=<rgba 0 0.7188 1 1>}
  NamedRerouteDeclaration(None=FunctionInput(Preview=Constant4Vector().RGBA) {bUsePreviewValueAsDefault=true, inputName=IsState, inputType=FunctionInput_Vector4, sortPriority=4}) {name=IsState, nodeColor=<rgba 0.1497 1 0.1633 1>}
  n193 = Subtract(
   A=Time()
   B=ScalarParameter() {defaultValue=-9999999, group=Wrap_BadBear_reactive, parameterName=TimeFrom-BP----------------, sortPriority=600}
  )
  n194 = ScalarParameter() {defaultValue=0.8, group=Wrap_BadBear_reactive, parameterName=BadBear_Time_Duration, sortPriority=601}
  NamedRerouteDeclaration(None=Saturate(None=Divide(A=@n193, B=@n194))) {name=reactive, nodeColor=<rgba 0.3594 1 0 1>}
  NamedRerouteDeclaration(None=@n67) {name=PreskinnedMap, nodeColor=<rgba 1 0 0.0703 1>}
  NamedRerouteDeclaration(None=@n9.Uvs) {name=UV_VP, nodeColor=<rgba 0.6406 1 0 1>}
  NamedRerouteDeclaration(None=ScalarParameter() {defaultValue=0.25, group=Wrap_BadBear_Lines, parameterName=BadBear_Line_Strength_Mst}) {name=Line_Strength, nodeColor=<rgba 0.3594 1 0 1>}
  NamedRerouteDeclaration(None=FunctionInput(Preview=Constant2Vector()) {bUsePreviewValueAsDefault=true, inputName=IsState_B, inputType=FunctionInput_Vector2, sortPriority=4}) {name=IsState_B, nodeColor=<rgba 0.1497 1 0.1633 1>}
  n195 = PreSkinnedLocalBounds()
  n196 = Subtract(A=@n179.PS, B=@n195.Min)
  n197 = MaterialFunctionCall(Color=Subtract(A=@n195.Max, B=@n195.Max)) {materialFunction=&SplitComponents}
  n198 = Max(A=Max(A=@n197.R, B=@n197.R), B=@n197.B)
  NamedRerouteDeclaration(None=FunctionInput(Preview=Divide(A=@n196, B=@n198)) {bUsePreviewValueAsDefault=true, inputName=Calcated_Bounds, sortPriority=11}) {name=Caclcated_Bounds, nodeColor=<rgba 0 0.9297 1 1>}
  n199 = TextureObjectParameter() {group=Wrap_BadBear, parameterName=BadBear__Tex_BC, sortPriority=5, texture=&T_BlacksRGB}
  NamedRerouteDeclaration(None=@n199) {name=Tex_BC, nodeColor=<rgba 0.7344 0 1 1>}
  n200 = PreSkinnedNormal()
  NamedRerouteDeclaration(None=FunctionInput(Preview=VertexInterpolator(VS=@n200).PS) {bUsePreviewValueAsDefault=true, inputName=Pre-Skinned Local Normal, sortPriority=10}) {name=PreSkin_Normal, nodeColor=<rgba 0.8281 1 0 1>}
  NamedRerouteDeclaration(None=@n18.Normal) {name=Root_Normal, nodeColor=<rgba 0.4062 0 1 1>}
  NamedRerouteUsage()
  NamedRerouteUsage()
  NamedRerouteUsage()
  n201 = Add(A=ComponentMask(None=@n120) {g=true}) {constB=-0.5}
  OneMinus(None=Saturate(None=Divide(A=@n201) {constB=0.001}))
  n202 = Add(A=ComponentMask(None=@n134) {g=true}) {constB=-0.5}
  OneMinus(None=Saturate(None=Divide(A=@n202) {constB=0.001}))
  n203 = Add(A=ComponentMask(None=@n148) {g=true}) {constB=-0.5}
  OneMinus(None=Saturate(None=Divide(A=@n203) {constB=0.001}))
  OneMinus(None=@n27)
  PixelNormalWS()
  Power(Base=@n24, Exp=ScalarParameter() {defaultValue=1.5, group=Wrap_BadBear_InnerGlow, parameterName=BadBear_Inner_Glow_Falloff})
  n204 = ComponentMask(None=@n123) {g=true}
  Saturate(None=Divide(A=Add(A=@n204) {constB=-0.5}) {constB=0.001})
  n205 = ComponentMask(None=@n138) {g=true}
  Saturate(None=Divide(A=Add(A=@n205) {constB=-0.5}) {constB=0.001})
  n206 = Multiply(A=Add(A=@n76.G) {constB=-0.5}) {constB=2}
  n207 = Add(
   A=ScalarParameter() {defaultValue=0.32, group=Wrap_BadBear, parameterName=BadBear_Base_Roughness, sortPriority=10}
   B=Saturate(None=Power(Base=@n206))
  )
  n208 = Power(Base=Multiply(A=OneMinus(None=@n90)) {constB=2})
  Saturate(None=Add(A=Saturate(None=@n207), B=Saturate(None=@n208)))
  n209 = ComponentMask(None=@n153) {g=true}
  Saturate(None=Divide(A=Add(A=@n209) {constB=-0.5}) {constB=0.001})
  ScalarParameter() {defaultValue=1, group=Wrap_BadBear_Subsurface, parameterName=BadBear_SubsSize}
  ScalarParameter() {defaultValue=1, group=Wrap_BadBear_Subsurface, parameterName=BadBear_SubsSize_B}
  n210 = PreSkinnedPosition()
  ShaderStageSwitch(PixelShader=VertexInterpolator(VS=@n210).PS, VertexShader=@n210)
 ]}
