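# Material function for the BeamU weapon-skin wrap, drawing animated UFO ships over a background and scrolling stars, switchable by static switch.
# Projected, rotated UVs form two tiled ship layers with random scale and sine motion; IsState selects per-object-type offsets and strengths.
# Emissive adds time-of-day-scaled rim, stars, and beam and light glows animated from timestamp parameters.
# Customization masks blend in the wrap and a two-color accent gradient.
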MaterialFunction /Game/Weapons/WeaponSkins/Wraps/MaterialFunctions/MF_Wrap_BeamU
graph {nodes [
  Add()
  Add()
  Constant2Vector() {r=1.1}
  n1 = GetMaterialAttributes(None=FunctionInput() {bUsePreviewValueAsDefault=true, inputName=Material, inputType=FunctionInput_MaterialAttributes, sortPriority=-20})
  n2 = NamedRerouteUsage()
  n3 = TextureObjectParameter() {group=Wrap__BeamU__Textures, parameterName=BU_BaseColor_01, sortPriority=11, texture=&T_BlacksRGB}
  n4 = ScalarParameter() {defaultValue=4, group=Wrap_BeamU, parameterName=BU_Ship_A_Number}
  n5 = Constant() {r=1}
  n6 = VectorParameter() {group=Wrap_BeamU_UVs, parameterName=BU_Ship_A_Offset_Vehicle}
  n7 = VectorParameter() {group=Wrap_BeamU_UVs, parameterName=BU_Ship_A_Offset_VehicleSide}
  n8 = FunctionInput() {bUsePreviewValueAsDefault=true, inputName=IsState, inputType=FunctionInput_Vector4, sortPriority=4}
  n9 = MaterialFunctionCall(Float4=@n8) {materialFunction=&BreakOutFloat4Components}
  n10 = LinearInterpolate(
   A=LinearInterpolate(
    A=VectorParameter() {group=Wrap_BeamU_UVs, parameterName=BU_Ship_A_Offset_Weapon}.RGB
    B=LinearInterpolate(A=@n6.RGB, B=@n7.RGB, Alpha=@n9.G)
    Alpha=@n9.R
   )
   B=LinearInterpolate(
    A=VectorParameter() {group=Wrap_BeamU_UVs, parameterName=BU_Ship_A_Offset_Mech}.RGB
    B=VectorParameter() {group=Wrap_BeamU_UVs, parameterName=BU_Ship_A_Offset_Character}.RGB
    Alpha=@n9.A
   )
   Alpha=@n9.B
  )
  n11 = MaterialFunctionCall(Color=@n8) {materialFunction=&SplitComponents}
  n12 = VertexInterpolator(VS=PreSkinnedPosition())
  n13 = PreSkinnedLocalBounds()
  n14 = FunctionInput(Preview=Divide(A=Subtract(A=@n12.PS, B=@n13.Min), B=Subtract(A=@n13.Max, B=@n13.Max))) {bUsePreviewValueAsDefault=true, inputName=Calcated_Bounds, sortPriority=11}
  n15 = FunctionInput(Preview=VertexInterpolator(VS=PreSkinnedNormal()).PS) {bUsePreviewValueAsDefault=true, inputName=Pre-Skinned Local Normal, sortPriority=10}
  n16 = MaterialFunctionCall!(
   For_Safe_Zones
   Normal_Map
   Caculated_Bounds=@n14
   Pre_Skinned_Normals=@n15
   Is Vehicle=@n11.R
   Is VehicleSide=@n11.R
   Is Mech=@n11.R
   Vehicle_Rotation
   UV_Rotation Bypass
   Mech_Rotation
   Uvs_Offset
   Offset-after rotation
   Uvs_TileRate
  ) {materialFunction=&/MF_Projection_Side_Normal}
  n17 = ScalarParameter() {group=Wrap_BeamU, parameterName=BU_Ship_Rotate_B_Weapon}
  n18 = MaterialFunctionCall(Float4=@n8) {materialFunction=&BreakOutFloat4Components}
  n19 = LinearInterpolate(
   A=ScalarParameter() {group=Wrap_BeamU, parameterName=BU_Ship_Rotate_B_Vehicle}
   B=ScalarParameter() {group=Wrap_BeamU, parameterName=BU_Ship_Rotate_B_Vehicle_Side}
   Alpha=@n18.G
  )
  n20 = ScalarParameter() {group=Wrap_BeamU, parameterName=BU_Ship_Rotate_B_Mech}
  n21 = ScalarParameter() {group=Wrap_BeamU, parameterName=BU_Ship_Rotate_B_Mech}
  n22 = MaterialFunctionCall(
   UVs=@n16.UVs No Offsets/Tiling
   Rotation Center=Constant() {r=0.5}
   "Rotation Angle (0-1)"=LinearInterpolate(A=LinearInterpolate(A=@n17, B=@n19, Alpha=@n18.R), B=LinearInterpolate(A=@n20, B=@n21, Alpha=@n18.A), Alpha=@n18.B)
  ) {materialFunction=&CustomRotator}
  n23 = Multiply(A=Add(A=ComponentMask(None=@n10) {g=true, r=true}, B=@n22.Rotated Values), B=AppendVector(A=@n4, B=@n5))
  n24 = Multiply(A=Add(A=Constant2Vector() {g=-0.4}, B=Frac(None=@n23)) {constA=1}, B=AppendVector(A=@n5, B=@n4))
  n25 = Constant() {r=1}
  n26 = ScalarParameter() {defaultValue=0.8, group=Wrap_BeamU, parameterName=BU_Ship_A_Scale}
  n27 = VectorParameter() {group=Wrap_BeamU, parameterName=BU_Ship_A_Random_seed, defaultValue=<rgba 3 6 0 0>}
  n28 = MaterialFunctionCall!(2 Vector=Add(A=Floor(None=@n23), B=ComponentMask(None=@n27.RGB) {g=true, r=true}), RandomSeed) {materialFunction=&/MF_Random_2}
  n29 = AppendVector(A=Constant(), B=@n28.utput)
  n30 = ComponentMask(None=VectorParameter() {group=Wrap_BeamU, parameterName=BU_Ship_A_Speed, defaultValue=<rgba 0 0.2 0 1>}.RGB) {g=true, r=true}
  n31 = Add(A=@n29, B=Multiply(A=Time(), B=@n30))
  n32 = MaterialFunctionCall(Float4=@n8) {materialFunction=&BreakOutFloat4Components}
  n33 = LinearInterpolate(
   A=ScalarParameter() {defaultValue=0.1, group=Wrap_BeamU, parameterName=BU_Ship_Sine_Strenght_A_Weapon}
   B=LinearInterpolate(
    A=ScalarParameter() {defaultValue=1, group=Wrap_BeamU, parameterName=BU_Ship_Sine_Strenght_A_Vehicle}
    B=ScalarParameter() {defaultValue=1, group=Wrap_BeamU, parameterName=BU_Ship_Sine_Strenght_A_Vehicle_Side}
    Alpha=@n32.G
   )
   Alpha=@n32.R
  )
  n34 = LinearInterpolate(
   A=ScalarParameter() {defaultValue=1, group=Wrap_BeamU, parameterName=BU_Ship_Sine_Strenght_A_Mech}
   B=ScalarParameter() {defaultValue=1, group=Wrap_BeamU, parameterName=BU_Ship_Sine_Strenght_A_Character}
   Alpha=@n32.A
  )
  n35 = Add(
   A=MaterialFunctionCall(UVs=@n24, Texture Scale=LinearInterpolate(A=@n25, B=@n26, Alpha=@n28.utput)) {materialFunction=&ScaleUVsByCenter}.UVs
   B=Multiply(A=Sine(None=@n31), B=LinearInterpolate(A=@n33, B=@n34, Alpha=@n32.B))
  )
  n36 = TextureSample!(UVs=@n35, Tex=@n3, Level=@n2, Apply View MipBias) {mipValueMode=TMVM_MipLevel, samplerSource=SSM_Clamp_WorldGroupSettings}
  n37 = Multiply(A=NamedRerouteUsage(), B=@n36.A)
  n38 = VectorParameter() {group=Wrap_BeamU_UVs, parameterName=BU_Ship_B_Offset_Vehicle}
  n39 = VectorParameter() {group=Wrap_BeamU_UVs, parameterName=BU_Ship_B_Offset_VehicleSide}
  n40 = MaterialFunctionCall(Float4=@n8) {materialFunction=&BreakOutFloat4Components}
  n41 = LinearInterpolate(
   A=LinearInterpolate(
    A=VectorParameter() {group=Wrap_BeamU_UVs, parameterName=BU_Ship_B_Offset_Weapon, defaultValue=<rgba 0.1 0 0 0>}.RGB
    B=LinearInterpolate(A=@n38.RGB, B=@n39.RGB, Alpha=@n40.G)
    Alpha=@n40.R
   )
   B=LinearInterpolate(
    A=VectorParameter() {group=Wrap_BeamU_UVs, parameterName=BU_Ship_B_Offset_Mech}.RGB
    B=VectorParameter() {group=Wrap_BeamU_UVs, parameterName=BU_Ship_B_Offset_Character}.RGB
    Alpha=@n40.A
   )
   Alpha=@n40.B
  )
  n42 = ScalarParameter() {defaultValue=4, group=Wrap_BeamU, parameterName=BU_Ship_B_Number}
  n43 = Constant() {r=1}
  n44 = Multiply(A=Add(A=@n22.Rotated Values, B=ComponentMask(None=@n41) {g=true, r=true}), B=AppendVector(A=@n42, B=@n43))
  n45 = Add(A=Constant2Vector() {g=-0.4}, B=Frac(None=@n44)) {constA=1}
  n46 = AppendVector(A=@n43, B=@n42)
  n47 = Constant() {r=1}
  n48 = ScalarParameter() {defaultValue=0.5, group=Wrap_BeamU, parameterName=BU_Ship_B_Scale}
  n49 = VectorParameter() {group=Wrap_BeamU, parameterName=BU_Ship_B_Random_seed, defaultValue=<rgba 3 6 0 0>}
  n50 = MaterialFunctionCall(2 Vector=Add(A=Floor(None=@n44), B=ComponentMask(None=@n49.RGB) {g=true, r=true})) {materialFunction=&MF_Random_2}
  n51 = AppendVector(A=Constant(), B=@n50.utput)
  n52 = ComponentMask(None=VectorParameter() {group=Wrap_BeamU, parameterName=BU_Ship_B_Speed, defaultValue=<rgba 0 0.023 0 1>}.RGB) {g=true, r=true}
  n53 = Add(A=@n51, B=Multiply(A=Time(), B=@n52))
  n54 = MaterialFunctionCall(Float4=@n8) {materialFunction=&BreakOutFloat4Components}
  n55 = LinearInterpolate(
   A=ScalarParameter() {defaultValue=0.5, group=Wrap_BeamU, parameterName=BU_Ship_Sine_Strenght_B_Weapon}
   B=LinearInterpolate(
    A=ScalarParameter() {defaultValue=1, group=Wrap_BeamU, parameterName=BU_Ship_Sine_Strenght_B_Vehicle}
    B=ScalarParameter() {defaultValue=1, group=Wrap_BeamU, parameterName=BU_Ship_Sine_Strenght_B_Vehicle_Side}
    Alpha=@n54.G
   )
   Alpha=@n54.R
  )
  n56 = LinearInterpolate(
   A=ScalarParameter() {defaultValue=1, group=Wrap_BeamU, parameterName=BU_Ship_Sine_Strenght_B_Mech}
   B=ScalarParameter() {defaultValue=1, group=Wrap_BeamU, parameterName=BU_Ship_Sine_Strenght_B_Character}
   Alpha=@n54.A
  )
  n57 = Add(
   A=MaterialFunctionCall(UVs=Multiply(A=@n45, B=@n46), Texture Scale=LinearInterpolate(A=@n47, B=@n48, Alpha=@n50.utput)) {materialFunction=&ScaleUVsByCenter}.UVs
   B=Multiply(A=Sine(None=@n53), B=LinearInterpolate(A=@n55, B=@n56, Alpha=@n54.B))
  )
  n58 = TextureSample(UVs=@n57, Tex=@n3, Level=@n2) {mipValueMode=TMVM_MipLevel, samplerSource=SSM_Clamp_WorldGroupSettings}
  n59 = Round(None=@n50.utput)
  n60 = Multiply(A=NamedRerouteUsage(), B=Multiply(A=@n58.A, B=@n59) {desc=Random take out UFO})
  n61 = OneMinus(None=Saturate(None=Add(A=@n37, B=@n60)))
  n62 = TextureSample!(
   UVs=@n22.Rotated Values
   Tex=TextureObjectParameter() {group=Wrap__BeamU__Textures, parameterName=BU_BaseColor_BG, texture=&T_BlacksRGB}
   Apply View MipBias
  )
  n63 = Divide(
   A=Subtract(A=Time(), B=ScalarParameter() {defaultValue=-9999999, group=Wrap_BeamU_Reactive, parameterName=BU_TimeFrom-BP_Beam, sortPriority=-600})
   B=ScalarParameter() {defaultValue=1.5, group=Wrap_BeamU_Reactive, parameterName=BU_Beam_AnimationDuration, sortPriority=33}
  )
  n64 = NamedRerouteUsage()
  n65 = TextureObjectParameter() {group=Wrap__BeamU__Textures, parameterName=BU_Mask_01, texture=&T_BlacksRGB}
  n66 = Multiply(
   A=TextureSample!(UVs=@n57, Tex=@n65, Level=@n64, Apply View MipBias) {mipValueMode=TMVM_MipLevel, samplerSource=SSM_Clamp_WorldGroupSettings, samplerType=SAMPLERTYPE_Masks}.RGB
   B=@n59
  )
  n67 = OneMinus(None=@n37)
  n68 = Multiply(
   A=Add(
    A=TextureSample(UVs=@n35, Tex=@n65, Level=@n64) {mipValueMode=TMVM_MipLevel, samplerSource=SSM_Clamp_WorldGroupSettings, samplerType=SAMPLERTYPE_Masks}.RGB
    B=Multiply(A=@n66, B=@n67)
   )
   B=NamedRerouteUsage()
  )
  n69 = Multiply(
   A=LinearInterpolate(
    A=ScalarParameter() {defaultValue=1, group=Wrap_BeamU_Reactive, parameterName=BU_Beam_Brightness_Start}
    B=ScalarParameter() {group=Wrap_BeamU_Reactive, parameterName=BU_Beam_Brightness_End}
    Alpha=Saturate(None=@n63)
   )
   B=Multiply(
    A=VectorParameter() {group=Wrap_BeamU, parameterName=BU_EmisBeam_Color, defaultValue=<rgba 0.4745 0.8795 1 1>}.RGB
    B=ComponentMask(None=@n68) {r=true}
   )
  )
  n70 = Divide(
   A=Subtract(A=Time(), B=ScalarParameter() {defaultValue=-9999999, group=Wrap_BeamU_Reactive, parameterName=BU_TimeFrom-BP_Lights, sortPriority=-600})
   B=ScalarParameter() {defaultValue=1.5, group=Wrap_BeamU_Reactive, parameterName=BU_Lights_AnimationDuration, sortPriority=33}
  )
  n71 = Multiply(
   A=Multiply(
    A=ComponentMask(None=@n68) {g=true}
    B=VectorParameter() {group=Wrap_BeamU, parameterName=BU_EmisGlow_Color, defaultValue=<rgba 2 0.7917 1.6698 1>}.RGB
   )
   B=LinearInterpolate(
    A=ScalarParameter() {defaultValue=1, group=Wrap_BeamU_Reactive, parameterName=BU_Lights_Brightness_Start}
    B=ScalarParameter() {group=Wrap_BeamU_Reactive, parameterName=BU_Lights_Brightness_End}
    Alpha=Saturate(None=@n70)
   )
  )
  n72 = VectorParameter() {group=Wrap_BeamU, parameterName=BU_EmisRim_Color, defaultValue=<rgba 2 1.9341 1.9544 1>}
  n73 = ComponentMask(None=@n68) {b=true}
  n74 = OneMinus(None=@n60)
  n75 = ScalarParameter() {defaultValue=1, group=Wrap_BeamU_UVs, parameterName=BU_StarsTileRate}
  n76 = Time()
  n77 = ComponentMask(None=VectorParameter() {group=Wrap_BeamU, parameterName=BU_StarSpeed, defaultValue=<rgba -0.005 0.001 0 1>}.RGB) {g=true, r=true}
  n78 = TextureSample!(UVs=Add(A=Multiply(A=@n75, B=@n16.UVs No Offsets/Tiling), B=Multiply(A=@n76, B=@n77) {constB=0.01}), Tex=@n65, Apply View MipBias)
  n79 = NamedRerouteUsage()
  n80 = Add(A=Add(A=Add(A=@n69, B=@n71), B=Multiply(A=@n72.RGB, B=@n73)), B=Multiply(A=Multiply(A=@n67, B=@n74), B=Multiply(A=@n78.A, B=@n79)))
  n81 = MaterialFunctionCall(Value_Morning=Constant() {r=0.5}, Value_Day=Constant() {r=1}, Value_Evening=Constant() {r=0.5}, Value_Night=Constant() {r=0.25}) {materialFunction=&Scalar_Time_of_Day}
  n82 = Constant3Vector()
  n83 = Multiply(A=VectorParameter() {group=Wrap_BeamU, parameterName=BU_Ship_B_ColorTint, defaultValue=<rgba 0.5 0.5 0.5 1>}.RGB, B=@n58.RGB)
  n84 = SetMaterialAttributes!(
   MaterialAttributes=@n1.MaterialAttributes
   BaseColor=LinearInterpolate(A=LinearInterpolate(A=@n82, B=@n83, Alpha=@n60), B=@n36.RGB, Alpha=@n37) {desc=Preview this to see Ships moving}
   Roughness=@n62.A
   Specular=ScalarParameter() {defaultValue=0.5, group=Wrap_BeamU, parameterName=BU_Specular, sortPriority=0}
   EmissiveColor=Add(A=Multiply(A=@n61, B=@n62.RGB), B=Multiply(A=@n80, B=@n81.Result))
   Metallic=ScalarParameter() {group=Wrap_BeamU, parameterName=BU_Metallic, sortPriority=0}
  ) {attributeSetTypes [
    #1
    #2
    #3
    #4
    #5
   ]}
  n85 = FunctionInput(Preview=Constant() {r=1}) {
   bUsePreviewValueAsDefault=true
   desc=base customization
   inputName="Customization Mask (_CM B Top left)"
   inputType=FunctionInput_Scalar
   sortPriority=49
  }
  n86 = BlendMaterialAttributes(A=@n1.MaterialAttributes, B=@n84, Alpha=@n85) {vertexAttributeBlendType=UseA}
  n87 = ComponentMask(None=@n22.Rotated Values) {r=true}
  n88 = ScalarParameter() {group=Wrap_BeamU, parameterName=BU_Accent_blendPower}
  n89 = SetMaterialAttributes!(
   MaterialAttributes=GetMaterialAttributes(None=@n86).MaterialAttributes
   BaseColor=LinearInterpolate(
    A=VectorParameter() {group=Wrap_BeamU, parameterName=BU_Accent_Color_A, defaultValue=<rgba 0.9896 0.2989 0.2989 1>}.RGB
    B=VectorParameter() {group=Wrap_BeamU, parameterName=BU_Accent_Color_B, defaultValue=<rgba 0.2455 0.4037 0.9896 1>}.RGB
    Alpha=Power(Base=@n87, Exp=@n88)
   )
   Metallic=ScalarParameter() {group=Wrap_BeamU, parameterName=BU_Accent_Metal}
   Specular=ScalarParameter() {group=Wrap_BeamU, parameterName=BU_Accent_Spec}
   Roughness=ScalarParameter() {group=Wrap_BeamU, parameterName=BU_Accent_Rough}
  ) {attributeSetTypes [
    #1
    #5
    #3
    #2
   ]}
  n90 = FunctionInput(Preview=Constant() {r=1}) {
   bUsePreviewValueAsDefault=true
   desc=secondary
   inputName="Customization Mask 2 (_CM B Top Right)"
   inputType=FunctionInput_Scalar
   sortPriority=50
  }
  FunctionOutput(None=StaticSwitchParameter(True=BlendMaterialAttributes(A=@n86, B=@n89, Alpha=@n90) {vertexAttributeBlendType=UseA}, False=@n1.MaterialAttributes) {group=Customization_Wrap, parameterName=Wrap_BeamU})
  n91 = Constant()
  MaterialFunctionCall(
   R=FunctionInput(Preview=Constant()) {inputName=isVehicle, inputType=FunctionInput_Scalar, sortPriority=5}
   G=FunctionInput(Preview=@n91) {bUsePreviewValueAsDefault=true, inputName=Side Project Vehicle, inputType=FunctionInput_Scalar}
   B=FunctionInput(Preview=@n91) {bUsePreviewValueAsDefault=true, inputName=isMech, inputType=FunctionInput_Scalar, sortPriority=4}
  ) {materialFunction=&AppendMany}
  Multiply(A=Add(A=@n29) {constB=0.75}, B=@n30)
  Multiply(A=Add(A=@n51) {constB=0.75}, B=@n52)
  Multiply(A=@n36.A, B=Round(None=@n28.utput) {desc=Random take out a UFO})
  Multiply(A=@n24) {constB=1.5}
  Multiply(A=@n28.utput, B=@n25)
  n92 = MaterialFunctionCall(Float3=Abs(None=VertexNormalWS())) {materialFunction=&BreakOutFloat3Components}
  n93 = ComponentMask(None=@n14) {b=true, desc=YZ, g=true}
  n94 = ComponentMask(None=@n14) {b=true, desc=XZ, r=true}
  Multiply(A=If(
    A=@n92.B
    B=Max(A=@n92.R, B=@n92.R)
    "A > B"=ComponentMask(None=@n14) {desc=YX, g=true, r=true}
    "A < B"=If(A=@n92.R, B=@n92.R, "A > B"=@n93, "A < B"=@n94)
   ))
  NamedRerouteDeclaration(None=MaterialFunctionCall!(UVs=@n16.UVs No Offsets/Tiling, Texture Size=ScalarParameter() {defaultValue=10000, group=Wrap_BeamU, parameterName=BU_Ship_Miplevel_Size}) {materialFunction=/Engine/Functions/Engine_MaterialFunctions01/Texturing/ComputeMipLevel}.MipLevel) {name=MipLevels, nodeColor=<rgba 1 0 0.6562 1>}
  NamedRerouteDeclaration() {name=test, nodeColor=<rgba 0.1 0.205 1 1>}
  n95 = MaterialFunctionCall(Color=Abs(None=@n15)) {materialFunction=&SplitComponents}
  n96 = MaterialFunctionCall(Color=@n8) {materialFunction=&SplitComponents}
  n97 = LinearInterpolate(A=@n95.R, B=LinearInterpolate(A=@n95.B, B=@n95.B, Alpha=@n96.G), Alpha=@n96.R)
  NamedRerouteDeclaration(None=MaterialFunctionCall(
    In=LinearInterpolate(A=@n97, B=@n95.G, Alpha=@n96.B)
    Contrast=ScalarParameter() {defaultValue=5, group=Wrap_BeamU, parameterName=BU_Alt+ZMask_Contrast, sortPriority=150}
   ) {materialFunction=&CheapContrast}.Result) {name=TriMask, nodeColor=<rgba 0 0.9531 1 1>}
  ScalarParameter() {defaultValue=0.2, group=Wrap_BeamU, parameterName=BU_Roghness, sortPriority=0}
  Subtract(A=@n62.RGB)
  VectorParameter() {group=Wrap_BeamU, parameterName=BU_BaseColor, sortPriority=0, defaultValue=<rgba 0 0 0 1>}
 ]}
