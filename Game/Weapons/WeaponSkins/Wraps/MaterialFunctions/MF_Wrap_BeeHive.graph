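# Material function for the BeeHive cosmetic wrap, blended onto the input material attributes through customization masks A and B behind the Wrap_BeeHive static switch.
# Side-projected UVs, tuned per isState (weapon, vehicle, mech, character), drive a background, two panning sine-wobbled bee layers and animated honey drips.
# Drip height feeds curve-atlas colors and perturbed normals, which warp camera-facing virtual-plane UVs for a fake inner glow.
# Emissive is edge-boosted by curvature and scaled by time of day.
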MaterialFunction /Game/Weapons/WeaponSkins/Wraps/MaterialFunctions/MF_Wrap_BeeHive
graph {nodes [
  n1 = PreSkinnedLocalBounds()
  n2 = Add(
   A=@n1.Min
   B=FunctionInput(Preview=Constant3Vector() {constant=<rgba 0 5 0 0>}) {bUsePreviewValueAsDefault=true, inputName=Object Bounds Min Adjustment, sortPriority=50}
  ) {constB=0}
  n3 = FunctionInput(Preview=Constant3Vector() {constant=<rgba -5 -25 -5 0>}) {bUsePreviewValueAsDefault=true, inputName=Object Bounds Max Adjustment, sortPriority=50}
  n4 = Multiply(A=Subtract(A=Add(A=@n1.Max, B=@n3) {constB=0}, B=@n2)) {constB=0.5}
  n5 = FunctionInput(Preview=Constant()) {bUsePreviewValueAsDefault=true, inputName=isState, inputType=FunctionInput_Vector4, sortPriority=4}
  n6 = MaterialFunctionCall(Float4=@n5) {materialFunction=&BreakOutFloat4Components}
  n7 = LinearInterpolate(A=TransformPosition(None=Add(A=@n2, B=@n4)) {transformType=TRANSFORMPOSSOURCE_World}, B=ObjectPositionWS(), Alpha=@n6.B)
  n8 = Normalize(VectorInput=Subtract(A=@n7, B=CameraPositionWS()))
  n9 = CrossProduct(A=@n8, B=Constant3Vector() {constant=<rgba 0 0 1 1>})
  n10 = MaterialFunctionCall() {materialFunction=/Engine/Functions/Engine_MaterialFunctions02/WorldPositionOffset/ObjectScale}
  n11 = MaterialFunctionCall(Float3=@n1.Extents) {materialFunction=&BreakOutFloat3Components}
  n12 = Max(A=Max(A=@n11.R, B=@n11.R), B=@n11.B)
  n13 = FunctionInput(Preview=Constant4Vector() {constant=<rgba 0 0 0 1>}.RGBA) {bUsePreviewValueAsDefault=true, inputName=Object-SafeZone-Virtual-Pivot, inputType=FunctionInput_Vector4, sortPriority=105}
  n14 = FunctionInput() {bUsePreviewValueAsDefault=true, inputName=Object-SafeZone-Virtual-CH-Switch, inputType=FunctionInput_Scalar, sortPriority=105}
  n15 = MaterialFunctionCall!(
   Camera _pos
   Axis
   Scale
   Scale_Multplyer
   Object-SafeZone-Virtual-Pivot=@n13
   Object-SafeZone-Virtual-CH-Switch=@n14
   SinglePlane Normal Axis
   SinglePlane X Axis
   PlaneCenter_distortion
  ) {materialFunction=&/MF_SafeZone_VirtualPlane}
  n16 = LinearInterpolate(
   A=MaterialFunctionCall!(Plane Normal Axis=@n8, Plane X Axis=Normalize(VectorInput=@n9), Plane Center=@n7, UV Scale=Multiply(A=@n10.Scale X, B=@n12)) {materialFunction=/Game/Athena/Environments/Nexus/Materials/VirtualPlaneCoordinatesAbs}.UVs
   B=@n15.Uvs
   Alpha=NamedRerouteUsage()
  )
  Add(
   A=@n16
   B=ComponentMask(None=VectorParameter() {group=Wrap_BeeHive, parameterName=BHive_Inner_Wrap_offset, defaultValue=<rgba 0 0 0 1>}.RGB) {g=true, r=true}
  )
  n17 = TextureObjectParameter() {group=Wrap_BeeHive, parameterName=Bhive_BackGrnd__Tex_N, samplerType=SAMPLERTYPE_Normal, texture=/Engine/EngineMaterials/FlatNormal}
  n18 = MaterialFunctionCall(Color=@n5) {materialFunction=&SplitComponents}
  n19 = ScalarParameter() {defaultValue=1, group=Wrap_BeeHive, parameterName=BHive_BackGrnd_Scale_Weapon}
  n20 = MaterialFunctionCall(Float4=@n5) {materialFunction=&BreakOutFloat4Components}
  n21 = LinearInterpolate(
   A=ScalarParameter() {defaultValue=1, group=Wrap_BeeHive, parameterName=BHive_BackGrnd_Scale_Vehicle}
   B=ScalarParameter() {defaultValue=1, group=Wrap_BeeHive, parameterName=BHive_BackGrnd_Scale_VehicleSide}
   Alpha=@n20.G
  )
  n22 = ScalarParameter() {defaultValue=1, group=Wrap_BeeHive, parameterName=BHive_BackGrnd_Scale_Mech}
  n23 = ScalarParameter() {defaultValue=1, group=Wrap_BeeHive, parameterName=BHive_BackGrnd_Scale_Character}
  n24 = PreSkinnedLocalBounds()
  n25 = Subtract(A=VertexInterpolator(VS=PreSkinnedPosition()).PS, B=@n24.Min)
  n26 = Subtract(A=@n24.Max, B=@n24.Max)
  n27 = FunctionInput(Preview=VertexInterpolator(VS=PreSkinnedNormal()).PS) {bUsePreviewValueAsDefault=true, inputName=Pre-Skinned Local Normal, sortPriority=10}
  n28 = MaterialFunctionCall!(
   For_Safe_Zones
   Normal_Map=@n17
   Caculated_Bounds=FunctionInput(Preview=Divide(A=@n25, B=@n26)) {bUsePreviewValueAsDefault=true, inputName=Calcated_Bounds, sortPriority=11}
   Pre_Skinned_Normals=@n27
   Is Vehicle=@n18.R
   Is VehicleSide=@n18.R
   Is Mech=@n18.R
   Vehicle_Rotation
   UV_Rotation Bypass
   Mech_Rotation
   Uvs_Offset=Constant()
   Offset-after rotation
   Uvs_TileRate=LinearInterpolate(A=LinearInterpolate(A=@n19, B=@n21, Alpha=@n20.R), B=LinearInterpolate(A=@n22, B=@n23, Alpha=@n20.A), Alpha=@n20.B)
  ) {materialFunction=&/MF_Projection_Side_Normal}
  Add(
   A=Multiply(A=ComponentMask(None=@n28.Raw_Normal) {g=true, r=true}, B=ScalarParameter() {group=Wrap_BeeHive, parameterName=BHive_Inner_Warp_Amount})
   B=@n16
  )
  Constant() {r=1}
  FunctionInput(Preview=Constant3Vector() {constant=<rgba 0 1 0 0>}) {bUsePreviewValueAsDefault=true, inputName=ForwardAxis}
  n29 = VectorParameter() {group=Wrap_BeeHive, parameterName=BHive_BaseColor_2nd, sortPriority=2, defaultValue=<rgba 0 0 0 1>}
  n30 = FunctionInput() {bUsePreviewValueAsDefault=true, inputName=Material, inputType=FunctionInput_MaterialAttributes, sortPriority=-20}
  n31 = GetMaterialAttributes(None=@n30) {attributeGetTypes [#1]}
  n32 = SetMaterialAttributes!(
   MaterialAttributes=@n31.MaterialAttributes
   BaseColor=@n29.RGB
   Roughness=ScalarParameter() {defaultValue=0.2, group=Wrap_BeeHive, parameterName=BHive_Roghness_2nd, sortPriority=2}
   Specular=ScalarParameter() {defaultValue=0.5, group=Wrap_BeeHive, parameterName=BHive_Specular_2nd, sortPriority=2}
   Metallic=ScalarParameter() {group=Wrap_BeeHive, parameterName=BHive_Metallic_2nd, sortPriority=2}
  ) {attributeSetTypes [
    #2
    #3
    #4
    #5
   ]}
  n33 = VectorParameter() {group=Wrap_BeeHive, parameterName=BHive_BaseColor, sortPriority=0, defaultValue=<rgba 0 0 0 1>}
  n34 = ScalarParameter() {defaultValue=2, group=Wrap_BeeHive, parameterName=BHive_Bees_B_TileRate}
  n35 = MaterialFunctionCall(Float4=@n5) {materialFunction=&BreakOutFloat4Components}
  n36 = LinearInterpolate(
   A=ScalarParameter() {defaultValue=2, group=Wrap_BeeHive, parameterName=BHive_Bees_B_TileRate_Vehicle}
   B=ScalarParameter() {defaultValue=2, group=Wrap_BeeHive, parameterName=BHive_Bees_B_TileRate_VehicleSide}
   Alpha=@n35.G
  )
  n37 = ScalarParameter() {defaultValue=2, group=Wrap_BeeHive, parameterName=BHive_Bees_B_TileRate_Mech}
  n38 = ScalarParameter() {defaultValue=2, group=Wrap_BeeHive, parameterName=BHive_Bees_B_TileRate_Character}
  n39 = Multiply(
   A=@n28.UVs No Offsets/Tiling
   B=LinearInterpolate(A=LinearInterpolate(A=@n34, B=@n36, Alpha=@n35.R), B=LinearInterpolate(A=@n37, B=@n38, Alpha=@n35.A), Alpha=@n35.B)
  ) {constB=2}
  n40 = Constant2Vector() {g=0.5, r=0.5}
  n41 = ScalarParameter() {group=Wrap_BeeHive, parameterName=BHive_Bees_B_Rotation}
  n42 = Multiply(A=Time(), B=ScalarParameter() {group=Wrap_BeeHive, parameterName=BHive_Bees_BY_Speed})
  n43 = Time()
  n44 = ScalarParameter() {group=Wrap_BeeHive, parameterName=BHive_Bees_BX_Sine_Speed}
  n45 = Multiply(A=Sine(None=Multiply(A=@n43, B=@n44)), B=ScalarParameter() {group=Wrap_BeeHive, parameterName=BHive_Bees_BX_Sine_strength}) {constB=0.5}
  n46 = TextureObjectParameter() {group=Wrap_BeeHive, parameterName=BHive_Bees__Tex, texture=&T_BlacksRGB}
  n47 = TextureSample(
   UVs=Add(
    A=MaterialFunctionCall(UVs=@n39, Rotation Center=@n40, "Rotation Angle (0-1)"=@n41) {materialFunction=&CustomRotator}.Rotated Values
    B=AppendVector(A=@n42, B=@n45)
   )
   Tex=@n46
  )
  n48 = MaterialFunctionCall(Color=Abs(None=@n27)) {materialFunction=&SplitComponents}
  n49 = MaterialFunctionCall(Color=@n5) {materialFunction=&SplitComponents}
  n50 = LinearInterpolate(A=@n48.B, B=@n48.B, Alpha=@n49.G)
  n51 = MaterialFunctionCall(
   In=LinearInterpolate(A=LinearInterpolate(A=@n48.R, B=@n50, Alpha=@n49.R), B=@n48.G, Alpha=@n49.B)
   Contrast=ScalarParameter() {group=Wrap_BeeHive, parameterName=BHive_Drip__TriplannerMask_Contrast}
  ) {materialFunction=&CheapContrast}
  n52 = Multiply(A=@n51.Result, B=@n47.A)
  n53 = ScalarParameter() {defaultValue=2, group=Wrap_BeeHive, parameterName=BHive_Bees_A_TileRate}
  n54 = MaterialFunctionCall(Float4=@n5) {materialFunction=&BreakOutFloat4Components}
  n55 = LinearInterpolate(
   A=ScalarParameter() {defaultValue=2, group=Wrap_BeeHive, parameterName=BHive_Bees_A_TileRate_Vehicle}
   B=ScalarParameter() {defaultValue=2, group=Wrap_BeeHive, parameterName=BHive_Bees_A_TileRate_VehicleSide}
   Alpha=@n54.G
  )
  n56 = ScalarParameter() {defaultValue=2, group=Wrap_BeeHive, parameterName=BHive_Bees_A_TileRate_Mech}
  n57 = ScalarParameter() {defaultValue=2, group=Wrap_BeeHive, parameterName=BHive_Bees_A_TileRate_Character}
  n58 = Multiply(
   A=@n28.UVs No Offsets/Tiling
   B=LinearInterpolate(A=LinearInterpolate(A=@n53, B=@n55, Alpha=@n54.R), B=LinearInterpolate(A=@n56, B=@n57, Alpha=@n54.A), Alpha=@n54.B)
  ) {constB=2}
  n59 = Constant2Vector() {g=0.5, r=0.5}
  n60 = ScalarParameter() {group=Wrap_BeeHive, parameterName=BHive_Bees_A_Rotation}
  n61 = Multiply(A=Time(), B=ScalarParameter() {group=Wrap_BeeHive, parameterName=Bhive_Bees_AY_Speed})
  n62 = Time()
  n63 = ScalarParameter() {group=Wrap_BeeHive, parameterName=BHive_Bees_AX_Sine_Speed}
  n64 = Multiply(A=Sine(None=Multiply(A=@n62, B=@n63)), B=ScalarParameter() {group=Wrap_BeeHive, parameterName=BHive_Bees_AX_Sine_strength}) {constB=0.5}
  n65 = TextureSample(
   UVs=Add(
    A=MaterialFunctionCall(UVs=@n58, Rotation Center=@n59, "Rotation Angle (0-1)"=@n60) {materialFunction=&CustomRotator}.Rotated Values
    B=AppendVector(A=@n61, B=@n64)
   )
   Tex=@n46
  )
  n66 = Multiply(A=@n51.Result, B=@n65.A)
  n67 = ScalarParameter() {group=Wrap_BeeHive, parameterName=BHive_Honey_Rotation_Weapon}
  n68 = MaterialFunctionCall(Color=@n5) {materialFunction=&SplitComponents}
  n69 = LinearInterpolate(
   A=ScalarParameter() {group=Wrap_BeeHive, parameterName=BHive_Honey_Rotation_Vehicle}
   B=ScalarParameter() {group=Wrap_BeeHive, parameterName=BHive_Honey_Rotation_VehicleSide}
   Alpha=@n68.G
  )
  n70 = MaterialFunctionCall(
   UVs=@n28.UVs No Offsets/Tiling
   "Rotation Angle (0-1)"=LinearInterpolate(
    A=LinearInterpolate(A=@n67, B=@n69, Alpha=@n68.R)
    B=ScalarParameter() {defaultValue=0.25, group=Wrap_BeeHive, parameterName=BHive_Honey_Rotation_Mech}
    Alpha=@n68.B
   )
  ) {materialFunction=&CustomRotator}
  n71 = ScalarParameter() {defaultValue=-0.5, group=Wrap_BeeHive, parameterName=BHive_Honey_OffsetX_Weapon}
  n72 = MaterialFunctionCall(Color=@n5) {materialFunction=&SplitComponents}
  n73 = LinearInterpolate(
   A=ScalarParameter() {defaultValue=-0.5, group=Wrap_BeeHive, parameterName=BHive_Honey_OffsetX_Vehicle}
   B=ScalarParameter() {defaultValue=-0.5, group=Wrap_BeeHive, parameterName=BHive_Honey_OffsetX_VehicleSide}
   Alpha=@n72.G
  )
  n74 = Add(
   A=ComponentMask(None=@n70.Rotated Values) {r=true}
   B=LinearInterpolate(
    A=LinearInterpolate(A=@n71, B=@n73, Alpha=@n72.R)
    B=ScalarParameter() {defaultValue=-0.5, group=Wrap_BeeHive, parameterName=BHive_Honey_OffsetX_Mech}
    Alpha=@n72.B
   )
  )
  n75 = Saturate(None=Divide(A=@n74, B=ScalarParameter() {defaultValue=0.25, group=Wrap_BeeHive, parameterName=BHive_Honey_Falloff_X}) {constB=0.5})
  n76 = ComponentMask(None=@n70.Rotated Values) {g=true}
  n77 = ScalarParameter() {defaultValue=-0.5, group=Wrap_BeeHive, parameterName=BHive_Honey_OffsetY_Vehicle}
  n78 = ScalarParameter() {defaultValue=-0.5, group=Wrap_BeeHive, parameterName=BHive_Honey_OffsetY_VehicleSide}
  n79 = MaterialFunctionCall(Color=@n5) {materialFunction=&SplitComponents}
  n80 = LinearInterpolate(
   A=LinearInterpolate(
    A=ScalarParameter() {defaultValue=0.5, group=Wrap_BeeHive, parameterName=BHive_Honey_OffsetY_Weapon}
    B=LinearInterpolate(A=@n77, B=@n78, Alpha=@n79.G)
    Alpha=@n79.R
   )
   B=ScalarParameter() {defaultValue=-0.5, group=Wrap_BeeHive, parameterName=BHive_Honey_OffsetY_Mech}
   Alpha=@n79.B
  )
  n81 = Saturate(None=Divide(A=Add(A=@n76, B=@n80), B=ScalarParameter() {defaultValue=-0.25, group=Wrap_BeeHive, parameterName=BHive_Honey_Falloff_Y}) {constB=0.5})
  n82 = Time()
  n83 = VectorParameter() {group=Wrap_BeeHive, parameterName=BHive_Honey_Warp_Speed, defaultValue=<rgba 0.02 -0.02 0 1>}
  n84 = Add(
   A=ComponentMask(None=Multiply(A=@n82, B=@n83.RGB)) {g=true, r=true}
   B=Multiply(A=ScalarParameter() {defaultValue=1, group=Wrap_BeeHive, parameterName=BHive_Honey_Warp_TileRate}, B=@n28.UVs No Offsets/Tiling)
  )
  n85 = TextureObjectParameter() {group=Wrap_BeeHive, parameterName=BHive_Drips__Tex, texture=&T_BlacksRGB}
  n86 = MaterialFunctionCall(
   Transition Phase=Power(Base=Multiply(A=@n75, B=@n81)) {constExponent=3}
   Height Texture=Multiply(A=TextureSample(UVs=@n84, Tex=@n85).A, B=ScalarParameter() {defaultValue=1, group=Wrap_BeeHive, parameterName=BHive_Honey_Warp_Amount}) {constB=0.5}
   Contrast=Constant()
  ) {materialFunction=&HeightLerp}
  n87 = MaterialFunctionCall(Color=@n5) {materialFunction=&SplitComponents}
  n88 = LinearInterpolate(
   A=ScalarParameter() {group=Wrap_BeeHive, parameterName=Bhive_Drips_Rotation_Weapon}
   B=LinearInterpolate(
    A=ScalarParameter() {group=Wrap_BeeHive, parameterName=BHive_Drips_Rotation_Vehicle}
    B=ScalarParameter() {group=Wrap_BeeHive, parameterName=BHive_Drips_Rotation_VehicleSide}
    Alpha=@n87.G
   )
   Alpha=@n87.R
  )
  n89 = ScalarParameter() {defaultValue=0.25, group=Wrap_BeeHive, parameterName=BHive_Drips_Rotation_Mech}
  n90 = MaterialFunctionCall(Float4=@n5) {materialFunction=&BreakOutFloat4Components}
  n91 = LinearInterpolate(
   A=ScalarParameter() {defaultValue=1.5, group=Wrap_BeeHive, parameterName=BHive_Drips_Scale_Weapon}
   B=LinearInterpolate(
    A=ScalarParameter() {defaultValue=1, group=Wrap_BeeHive, parameterName=BHive_Drips_Scale_Vehicle}
    B=ScalarParameter() {defaultValue=1, group=Wrap_BeeHive, parameterName=BHive_Drips_Scale_VehicleSide}
    Alpha=@n90.G
   )
   Alpha=@n90.R
  )
  n92 = LinearInterpolate(
   A=ScalarParameter() {defaultValue=1, group=Wrap_BeeHive, parameterName=BHive_Drips_Scale_Mech}
   B=ScalarParameter() {defaultValue=1, group=Wrap_BeeHive, parameterName=BHive_Drips_Scale_Character}
   Alpha=@n90.A
  )
  n93 = Multiply(
   A=MaterialFunctionCall(UVs=@n28.UVs No Offsets/Tiling, "Rotation Angle (0-1)"=LinearInterpolate(A=@n88, B=@n89, Alpha=@n87.B)) {materialFunction=&CustomRotator}.Rotated Values
   B=Multiply(A=LinearInterpolate(A=@n91, B=@n92, Alpha=@n90.B), B=Constant2Vector() {g=0.75, r=1})
  )
  n94 = MaterialFunctionCall(Color=@n5) {materialFunction=&SplitComponents}
  n95 = LinearInterpolate(
   A=VectorParameter() {group=Wrap_BeeHive, parameterName=BHive_Drips_Offset_Weapon, defaultValue=<rgba -0.6 0 0 1>}.RGB
   B=LinearInterpolate(
    A=VectorParameter() {group=Wrap_BeeHive, parameterName=BHive_Drips_Offset_Vehicle, defaultValue=<rgba 0 0 0 1>}.RGB
    B=VectorParameter() {group=Wrap_BeeHive, parameterName=BHive_Drips_Offset_VehicleSide, defaultValue=<rgba 0 0 0 1>}.RGB
    Alpha=@n94.G
   )
   Alpha=@n94.R
  )
  n96 = VectorParameter() {group=Wrap_BeeHive, parameterName=BHive_Drips_Offset_Mech, defaultValue=<rgba 0 0 0 1>}
  n97 = Time()
  n98 = VectorParameter() {group=Wrap_BeeHive, parameterName=BHive_Drips_Speed, defaultValue=<rgba 0 -0.02 0 1>}
  n99 = ComponentMask(None=Add(A=LinearInterpolate(A=@n95, B=@n96.RGB, Alpha=@n94.B), B=Multiply(A=@n97, B=@n98.RGB))) {g=true, r=true}
  n100 = Multiply(A=@n51.Result, B=TextureSample(UVs=Add(A=@n93, B=@n99), Tex=@n85).R)
  n101 = Saturate(None=Divide(A=@n74) {constB=0.1})
  n102 = Multiply(A=MaterialFunctionCall(Transition Phase=LinearInterpolate(Alpha=@n86.Alpha) {constA=0.45}, Height Texture=Multiply(A=@n100, B=@n101)) {materialFunction=&HeightLerp}.Alpha)
  n103 = CurveAtlasRowParameter(CurveTime=@n102) {
   atlas=/Game/Weapons/WeaponSkins/Wraps/Textures/CruveAtlas_Wraps_01
   curve=/Game/Weapons/WeaponSkins/Wraps/Textures/FlapjackWrangler/CC_Wrap_FlapjackWrangler
   defaultValue=35
   group=Wrap_BeeHive
   parameterName="BHive_Honey_RemaptoColorCurve  A=Mask"
  }
  n104 = MaterialFunctionCall!(
   Index=ScalarParameter() {defaultValue=35, group=Wrap_BeeHive, parameterName="BHive_Honey_RemaptoColorCurve  A=Mask  Index"}
   CCinput=@n103
   CCinput_Alpha=@n103
   UV=@n102
  ) {materialFunction=&/MF_Global_CC_Tex}
  n105 = TextureSample(
   UVs=@n28.UVs
   Tex=TextureObjectParameter() {group=Wrap_BeeHive, parameterName=BHive_BackGrnd__Tex_BC, texture=&T_BlacksRGB}
  )
  n106 = Constant() {r=0.5}
  n107 = Constant() {r=1}
  n108 = Constant() {r=0.5}
  n109 = Constant() {r=0.25}
  n110 = ScalarParameter() {defaultValue=1, group=Wrap_BeeHive, parameterName=BHive_Honey_Emissive_Brightness}
  n111 = CurveAtlasRowParameter(CurveTime=@n102) {
   atlas=&CruveAtlas_Wraps_01
   curve=/Game/Weapons/WeaponSkins/Wraps/Textures/FlapjackWrangler/CC_Wrap_FlapjackWrangler01
   defaultValue=36
   group=Wrap_BeeHive
   parameterName=BHive_Honey_RemaptoColorCurve ForNormal
  }
  n112 = MaterialFunctionCall(
   Index=ScalarParameter() {defaultValue=36, group=Wrap_BeeHive, parameterName="BHive_Honey_RemaptoColorCurve ForNormal  Index"}
   CCinput=@n111
   CCinput_Alpha=@n111
   UV=@n102
  ) {materialFunction=&MF_Global_CC_Tex}
  n113 = ComponentMask(None=@n112.3V) {r=true}
  n114 = MaterialFunctionCall!(Value=Saturate(None=Multiply(A=@n104.Alpha, B=@n113))) {materialFunction=/Engine/Functions/Engine_MaterialFunctions03/Procedurals/PreparePerturbNormalHQ}
  n115 = MaterialFunctionCall!(
   Bump one pixel right=@n114.Value one pixel right
   Bump center=@n114.Value one pixel right
   Bump one pixel down=@n114.Value one pixel right
   WS Normal
  ) {materialFunction=/Engine/Functions/Engine_MaterialFunctions03/Procedurals/PerturbNormalHQ}
  n116 = LinearInterpolate(
   A=@n28.Raw_Normal
   B=Multiply(
    A=Transform(None=@n115.WS Normal) {transformSourceType=TRANSFORMSOURCE_World, transformType=TRANSFORM_Tangent}
    B=Constant3Vector() {constant=<rgba 1 1 0 0>}
   )
   Alpha=@n104.Alpha
  )
  n117 = Saturate(None=OneMinus(None=Add(A=@n66, B=@n52)))
  n118 = LinearInterpolate(
   A=@n31.Normal
   B=Add(A=@n31.Normal, B=Multiply(A=@n116, B=@n117))
   Alpha=ScalarParameter() {defaultValue=1, group=Wrap_BeeHive, parameterName=BHive_Honey_Normal_Intensty}
  )
  n119 = Add(
   A=Multiply(A=ComponentMask(None=@n118) {g=true, r=true}, B=ScalarParameter() {group=Wrap_BeeHive, parameterName=BHive_Honey_Inner_Glow_Warp})
   B=@n16
  )
  n120 = Constant() {r=0.5}
  n121 = ScalarParameter() {defaultValue=0.5, group=Wrap_BeeHive, parameterName=BHive_Honey_Inner_RadusWeapon}
  n122 = MaterialFunctionCall(Color=@n5) {materialFunction=&SplitComponents}
  n123 = LinearInterpolate(
   A=ScalarParameter() {defaultValue=0.5, group=Wrap_BeeHive, parameterName=BHive_Honey_Inner_Radus_Vehicle}
   B=ScalarParameter() {defaultValue=0.5, group=Wrap_BeeHive, parameterName=BHive_Honey_Inner_Radus_VehicleSide}
   Alpha=@n122.G
  )
  n124 = SphereMask(
   A=@n119
   B=@n120
   Radius=LinearInterpolate(
    A=LinearInterpolate(A=@n121, B=@n123, Alpha=@n122.R)
    B=ScalarParameter() {defaultValue=0.5, group=Wrap_BeeHive, parameterName=BHive_Honey_Inner_Radus_Mech}
    Alpha=@n122.B
   )
   Hardness=ScalarParameter() {group=Wrap_BeeHive, parameterName=BHive_Honey_Inner_Falloff}
  )
  n125 = Multiply(
   A=VectorParameter() {group=Wrap_BeeHive, parameterName=BHive_Honey_Inner_Color, defaultValue=<rgba 0.9531 0.7795 0.0173 1>}.RGB
   B=Saturate(None=OneMinus(None=@n124))
  )
  n126 = Multiply(A=Add(A=@n104.3V, B=Multiply(A=@n112.Alpha, B=@n125)), B=@n104.Alpha)
  n127 = VectorParameter() {group=Wrap_BeeHive, parameterName=Bhive_BackGrnd_Color, defaultValue=<rgba 1 1 1 1>}
  n128 = MaterialFunctionCall(In=@n105.A, Contrast=ScalarParameter() {defaultValue=2, group=Wrap_BeeHive, parameterName=BHive_BackGnd_Emissive_Edge_Contrast}) {materialFunction=&CheapContrast}
  n129 = ScalarParameter() {defaultValue=0.11, group=Wrap_BeeHive, parameterName=BHive_BackGnd_Emissive_Edge_Brightness}
  n130 = ScalarParameter() {group=Wrap_BeeHive, parameterName=BHive_Honey_Inner_Brightness_B}
  n131 = ScalarParameter() {defaultValue=1, group=Wrap_BeeHive, parameterName=BHive_Honey_Inner_Brightness_A}
  n132 = SphereMask(
   A=@n119
   B=@n120
   Radius=LinearInterpolate(
    A=ScalarParameter() {defaultValue=0.5, group=Wrap_BeeHive, parameterName=BHive_Inner_Radus}
    B=ScalarParameter() {defaultValue=0.5, group=Wrap_BeeHive, parameterName=BHive_Inner_Radus_Mech}
    Alpha=@n6.B
   )
   Hardness=LinearInterpolate(
    A=ScalarParameter() {group=Wrap_BeeHive, parameterName=BHive_Inner_Falloff}
    B=ScalarParameter() {group=Wrap_BeeHive, parameterName=BHive_Inner_Falloff_Mech}
    Alpha=@n6.B
   )
  )
  n133 = Multiply(A=Add(A=Multiply(A=@n128.Result, B=@n129), B=LinearInterpolate(A=@n130, B=@n131, Alpha=@n132)), B=@n105.RGB)
  n134 = Multiply(A=LinearInterpolate(A=Multiply(A=@n127.RGB, B=@n133), B=Multiply(A=@n110, B=@n126), Alpha=@n104.Alpha), B=@n117)
  n135 = Multiply(A=@n134, B=ScalarParameter() {defaultValue=1, group=Wrap_BeeHive, parameterName=BHive_Edge_Brightness, sortPriority=0})
  n136 = FunctionInput() {bUsePreviewValueAsDefault=true, inputName=Curavture, inputType=FunctionInput_Scalar, sortPriority=24}
  n137 = Constant() {r=1}
  n138 = Saturate(None=MaterialFunctionCall(Input=@n136, Input Low=Constant() {r=0.5}, Input High=@n137, Target Low=Constant(), Target High=@n137) {materialFunction=&RemapValueRange}.Result)
  n139 = SetMaterialAttributes!(
   MaterialAttributes=@n31.MaterialAttributes
   BaseColor=LinearInterpolate(A=LinearInterpolate(A=@n33.RGB, B=@n47.RGB, Alpha=@n52), B=@n65.RGB, Alpha=@n66)
   Roughness=LinearInterpolate(A=@n105.A, B=ScalarParameter() {defaultValue=0.2, group=Wrap_BeeHive, parameterName=BHive_Roghness, sortPriority=0}, Alpha=@n104.Alpha)
   Specular=LinearInterpolate(
    A=ScalarParameter() {defaultValue=0.5, group=Wrap_BeeHive, parameterName=BHive_Specular_Bck, sortPriority=0}
    B=ScalarParameter() {defaultValue=0.5, group=Wrap_BeeHive, parameterName=BHive_Specular, sortPriority=0}
    Alpha=@n104.Alpha
   )
   EmissiveColor=Multiply(
    A=LinearInterpolate(A=@n134, B=@n135, Alpha=@n138)
    B=MaterialFunctionCall(Value_Morning=@n106, Value_Day=@n107, Value_Evening=@n108, Value_Night=@n109) {materialFunction=&Scalar_Time_of_Day}.Result
   )
   Metallic=ScalarParameter() {group=Wrap_BeeHive, parameterName=BHive_Metallic, sortPriority=0}
   Normal=@n118
  ) {attributeSetTypes [
    #2
    #3
    #4
    #6
    #5
    #1
   ]}
  n140 = TextureSample() {samplerType=SAMPLERTYPE_Masks, texture=/Game/Weapons/FORT_Rifles/Textures/T_Scar_CM}
  n141 = FunctionInput(Preview=@n140.R) {bUsePreviewValueAsDefault=true, inputName=Customization Mask A, inputType=FunctionInput_Scalar, sortPriority=24}
  n142 = BlendMaterialAttributes(A=@n32, B=@n139, Alpha=@n141) {vertexAttributeBlendType=UseA}
  n143 = Saturate(None=Add(
    A=@n141
    B=FunctionInput(Preview=@n140.B) {bUsePreviewValueAsDefault=true, inputName=Customization Mask B, inputType=FunctionInput_Scalar, sortPriority=24}
   ))
  FunctionOutput(None=StaticSwitchParameter(True=BlendMaterialAttributes(A=@n31.MaterialAttributes, B=@n142, Alpha=@n143) {vertexAttributeBlendType=UseA}, False=@n30) {group=Customization_Wrap, parameterName=Wrap_BeeHive})
  Multiply(A=Add(A=@n136) {constB=-0.5}) {constB=2}
  NamedRerouteDeclaration(None=@n6.A) {name=IsCharacter, nodeColor=<rgba 0.8649 1 0.1197 1>}
  NamedRerouteDeclaration(None=@n15.Uvs) {name=SafeZoneVP_UVs, nodeColor=<rgba 1 0 0.6562 1>}
  OneMinus(None=@n138)
 ]}
