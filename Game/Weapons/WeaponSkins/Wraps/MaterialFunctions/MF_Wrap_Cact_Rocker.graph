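# Weapon-wrap material function for the Cactus Rocker skin, applied through a static switch over masked areas of the input material.
# Side projection from pre-skinned bounds and normals maps the pattern, with IsState inputs selecting placement for weapons, vehicles, mechs, characters, backpacks and pickaxes.
# EventAudio collection parameters drive animated lines, Fresnel glow and hue-shifting color.
# A separately rotated, scaled and clamped mouth projection builds masks and colored eye, eyebrow and teeth emissive.
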MaterialFunction /Game/Weapons/WeaponSkins/Wraps/MaterialFunctions/MF_Wrap_Cact_Rocker
graph {nodes [
  n1 = TextureSample() {samplerType=SAMPLERTYPE_Masks, texture=/Game/Weapons/FORT_Rifles/Textures/T_Scar_CM}
  n2 = FunctionInput(Preview=@n1.R) {bUsePreviewValueAsDefault=true, inputName=Mask Primary, inputType=FunctionInput_Scalar, sortPriority=24, previewValue=<1 0 0 1>}
  n3 = FunctionInput(Preview=@n1.B) {bUsePreviewValueAsDefault=true, inputName=Mask Secondary, inputType=FunctionInput_Scalar, sortPriority=24, previewValue=<1 0 0 1>}
  n4 = BlendMaterialAttributes(A=NamedRerouteUsage(), B=NamedRerouteUsage(), Alpha=Saturate(None=Add(A=@n2, B=@n3))) {vertexAttributeBlendType=UseA}
  BlendMaterialAttributes(A=@n4, Alpha=NamedRerouteDeclaration(None=@n3) {name=Accent_Msk, nodeColor=<rgba 0 0.6484 1 1>}) {vertexAttributeBlendType=UseA}
  Constant4Vector()
  Constant() {r=1}
  FunctionInput(Preview=@n1.G) {bUsePreviewValueAsDefault=true, inputName=Mask Crevise, inputType=FunctionInput_Scalar, sortPriority=24, previewValue=<1 0 0 1>}
  n5 = TextureSample() {samplerType=SAMPLERTYPE_Normal, texture=/Game/Weapons/FORT_Rifles/T_SCAR_N}
  n6 = Constant3Vector() {constant=<rgba 1 0 0 0>}
  n7 = GetMaterialAttributes(None=FunctionInput(Preview=SetMaterialAttributes!(MaterialAttributes, Normal=@n5.RGB, BaseColor=@n6) {attributeSetTypes [#1, #2]}) {inputName=Material, inputType=FunctionInput_MaterialAttributes}) {attributeGetTypes [
    #2
    #3
    #1
    #4
   ]}
  FunctionOutput(None=StaticSwitchParameter(True=@n4, False=NamedRerouteDeclaration(None=@n7.MaterialAttributes) {name=Root_Material, nodeColor=<rgba 0.7812 1 0 1>}) {desc=TopSide Projection Base - Emissive pan Mask, group=Customization_Wrap, parameterName=Wrap_Cactus_Rocker}) {outputName=Result_1}
  n8 = NamedRerouteUsage()
  n9 = MaterialFunctionCall(Float4=NamedRerouteUsage()) {materialFunction=&BreakOutFloat4Components}
  n10 = MaterialFunctionCall!(
   For_Safe_Zones
   Normal_Map=NamedRerouteUsage()
   Caculated_Bounds=NamedRerouteUsage()
   Pre_Skinned_Normals=NamedRerouteUsage()
   Is Vehicle=@n9.R
   Is VehicleSide=@n9.R
   Is Mech=@n9.R
   Vehicle_Rotation
   UV_Rotation Bypass=NamedRerouteUsage()
   Mech_Rotation
   Uvs_Offset=NamedRerouteUsage()
   Offset-after rotation
   Uvs_TileRate=NamedRerouteUsage()
  ) {materialFunction=&/MF_Projection_Side_Normal}
  n11 = ComponentMask(None=Multiply(A=@n10.UVs, B=ScalarParameter() {defaultValue=1, group=Wrap_CactusRocker_Audio, parameterName=CaRocker_ColorCurve_TileRate})) {r=true}
  n12 = NamedRerouteUsage()
  n13 = MaterialFunctionCall(Hue Shift Percentage=Frac(None=Panner(Coordinate=@n11, Time=@n12) {speedX=0.5}), Texture=Constant3Vector() {constant=<rgba 1 0 0 0>}) {materialFunction=&HueShift}
  n14 = NamedRerouteUsage()
  n15 = Multiply(
   A=Desaturation(
    None=MaterialFunctionCall!(Base=@n13.Result, Add=@n14) {desc=multiplyAdd by audio, materialFunction=/Engine/Functions/Engine_MaterialFunctions02/Math/MultiplyAdd}.Result
    Fraction=ScalarParameter() {defaultValue=0.5, group=Wrap_CactusRocker_Audio, parameterName=CaRocker_ColorCurve_Saturation}
   )
   B=ScalarParameter() {defaultValue=0.75, group=Wrap_CactusRocker_Audio, parameterName=CaRocker_ColorCurve_Brightness}
  )
  n16 = NamedRerouteUsage()
  n17 = Multiply(
   A=Fresnel(ExponentIn=ScalarParameter() {defaultValue=2, group=Wrap_CactusRocker_Audio, parameterName=CaRocker_BodyFresnelExp, sliderMax=8}) {baseReflectFraction=0, exponent=2}
   B=LinearInterpolate(
    A=ScalarParameter() {defaultValue=0.1, group=Wrap_CactusRocker_Audio, parameterName=CaRocker_BodyFresnel_Intensity, sliderMax=1}
    B=ScalarParameter() {defaultValue=0.5, group=Wrap_CactusRocker_Audio, parameterName=CaRocker_BodyFresnel_Intensity_Active, sliderMax=1}
    Alpha=Saturate(None=@n16)
   )
  )
  n18 = NamedRerouteUsage()
  n19 = OneMinus(None=NamedRerouteUsage())
  n20 = NamedRerouteUsage()
  n21 = Abs(None=Multiply(A=Add(A=@n20) {constB=-0.5}) {constB=2})
  n22 = Saturate(None=OneMinus(None=ComponentMask(None=@n21) {r=true}))
  n23 = Add(
   A=ComponentMask(None=@n20) {r=true}
   B=Multiply(A=Time(), B=ScalarParameter() {defaultValue=0.5, group=Wrap_CactusRocker_UVs, parameterName=CaRocker_Line_Audio_Speed})
  )
  n24 = ScalarParameter() {defaultValue=4, group=Wrap_CactusRocker_UVs, parameterName=CaRocker_Line_Audio_Tile_Rate}
  n25 = Sine(None=Frac(None=Multiply(A=@n23, B=@n24))) {period=2}
  n26 = Add(A=Multiply(A=@n22, B=Multiply(A=@n25, B=@n25))) {constB=0.25}
  n27 = ScalarParameter() {defaultValue=5, group=Wrap_CactusRocker_UVs, parameterName=CaRocker_Line_Audio_Strength}
  n28 = CollectionParameter() {collection=&FortniteMaterialParameters, parameterName=EventAudio_Spectrum}
  n29 = MaterialFunctionCall(Float4=@n28) {materialFunction=&BreakOutFloat4Components}
  n30 = Power(Base=Add(A=Multiply(A=@n29.R) {constB=0.75}, B=Multiply(A=@n29.B) {constB=0.75}))
  n31 = ScalarParameter() {defaultValue=0.5, group=Wrap_CactusRocker_UVs, parameterName=CaRocker_Line_Audio_Enflunce}
  n32 = NamedRerouteUsage()
  n33 = NamedRerouteUsage()
  n34 = Saturate(None=Add(A=ComponentMask(None=@n32) {r=true}, B=TextureSample(UVs=@n10.UVs, Tex=@n33) {samplerType=SAMPLERTYPE_Normal}.R))
  n35 = Multiply(A=LinearInterpolate(B=Multiply(A=@n26, B=@n27) {constB=5}, Alpha=Clamp(None=@n30, Max=@n31) {maxDefault=0.5}) {constA=1, constB=3}, B=@n34)
  n36 = Add(A=@n35, B=@n34)
  n37 = Saturate(None=MaterialFunctionCall(
    In=Multiply(A=@n35, B=@n36)
    Contrast=ScalarParameter() {defaultValue=10, group=Wrap_CactusRocker_UVs, parameterName=CaRocker_Line_Audio_Contrast}
   ) {materialFunction=&CheapContrast}.Result)
  n38 = Add(A=Multiply(A=@n18, B=@n19), B=@n37)
  n39 = Saturate(None=Add(A=@n38, B=@n17))
  LinearInterpolate(A=@n8, B=@n15, Alpha=@n39)
  LinearInterpolate(
   A=@n19
   Alpha=MaterialFunctionCall(Float4=NamedRerouteUsage()) {materialFunction=&BreakOutFloat4Components}.R
  )
  n40 = MaterialFunctionCall(Float4=NamedRerouteUsage()) {materialFunction=&BreakOutFloat4Components}
  MaterialFunctionCall(
   For_Safe_Zones=StaticBool() {value=true}
   Normal_Map=NamedRerouteUsage()
   Caculated_Bounds=NamedRerouteUsage()
   Pre_Skinned_Normals=NamedRerouteUsage()
   Is Vehicle=@n40.R
   Is VehicleSide=@n40.R
   Is Mech=@n40.R
   UV_Rotation Bypass=NamedRerouteUsage()
   Uvs_Offset=NamedRerouteUsage()
   Uvs_TileRate=NamedRerouteUsage()
  ) {materialFunction=&MF_Projection_Side_Normal}
  Multiply()
  n41 = Sine(None=Multiply(A=Time(), B=ScalarParameter() {defaultValue=0.5})) {period=0.25}
  Multiply(A=@n41, B=@n41)
  Multiply(A=@n13.Result, B=@n17)
  Multiply(A=@n22)
  NamedRerouteDeclaration(None=TextureObjectParameter() {group=Wrap_CactusRocker, parameterName=CaRocker__Text_Msk, texture=&T_BlacksRGB}) {name=Text_Msk, nodeColor=<rgba 0.6406 1 0 1>}
  NamedRerouteDeclaration(None=TextureObjectParameter() {group=Wrap_CactusRocker, parameterName=CaRocker__Text_BC, texture=&T_BlacksRGB}) {name=Text_BC, nodeColor=<rgba 0.6406 1 0 1>}
  n42 = ScalarParameter() {group=Wrap_CactusRocker_UVs, parameterName=CaRocker_Mouth_Rotation_Weapon}
  n43 = MaterialFunctionCall(Float4=NamedRerouteUsage()) {materialFunction=&BreakOutFloat4Components}
  n44 = LinearInterpolate(
   A=ScalarParameter() {group=Wrap_CactusRocker_UVs, parameterName=CaRocker_Mouth_Rotation_Vehicle}
   B=ScalarParameter() {group=Wrap_CactusRocker_UVs, parameterName=CaRocker_Mouth_Rotation_VehicleSide}
   Alpha=@n43.G
  )
  n45 = ScalarParameter() {group=Wrap_CactusRocker_UVs, parameterName=CaRocker_Mouth_Rotation_Mech}
  n46 = ScalarParameter() {group=Wrap_CactusRocker_UVs, parameterName=CaRocker_Mouth_Rotation_Character}
  NamedRerouteDeclaration(None=LinearInterpolate(A=LinearInterpolate(A=@n42, B=@n44, Alpha=@n43.R), B=LinearInterpolate(A=@n45, B=@n46, Alpha=@n43.A), Alpha=@n43.B)) {name=Side_Rotation_Mouth, nodeColor=<rgba 1 0.7266 0 1>}
  n47 = VectorParameter() {group=Wrap_CactusRocker_UVs, parameterName=CaRocker_Mouth_Scale_Weapon, defaultValue=<rgba 1.25 1 1 0.18>}
  n48 = VectorParameter() {group=Wrap_CactusRocker_UVs, parameterName=CaRocker_Mouth_Scale_Vehicle, defaultValue=<rgba 1 1 1 0.25>}
  n49 = Multiply(A=@n48.RGB, B=@n48.RGB)
  n50 = VectorParameter() {group=Wrap_CactusRocker_UVs, parameterName=CaRocker_Mouth_Scale_Vehicle_Side, defaultValue=<rgba 1 1 1 0.25>}
  n51 = Multiply(A=@n50.RGB, B=@n50.RGB)
  n52 = MaterialFunctionCall(Float4=NamedRerouteUsage()) {materialFunction=&BreakOutFloat4Components}
  n53 = VectorParameter() {group=Wrap_CactusRocker_UVs, parameterName=CaRocker_Mouth_Scale_Mech, defaultValue=<rgba 1 1 1 0.25>}
  n54 = VectorParameter() {group=Wrap_CactusRocker_UVs, parameterName=CaRocker_Mouth_Scale_Character, defaultValue=<rgba 1 1 1 0.25>}
  n55 = LinearInterpolate(
   A=LinearInterpolate(A=Multiply(A=@n47.RGB, B=@n47.RGB), B=LinearInterpolate(A=@n49, B=@n51, Alpha=@n52.G), Alpha=@n52.R)
   B=LinearInterpolate(A=Multiply(A=@n53.RGB, B=@n53.RGB), B=Multiply(A=@n54.RGB, B=@n54.RGB), Alpha=@n52.A)
   Alpha=@n52.B
  )
  n56 = VectorParameter() {group=Wrap_CactusRocker_UVs, parameterName=CaRocker_Mouth_Scale_Backpack, defaultValue=<rgba 1 1 1 0.25>}
  n57 = Multiply(A=@n56.RGB, B=@n56.RGB)
  n58 = MaterialFunctionCall(Float2=NamedRerouteUsage()) {materialFunction=&BreakOutFloat2Components}
  n59 = VectorParameter() {group=Wrap_CactusRocker_UVs, parameterName=CaRocker_Mouth_Scale_Pickaxe, defaultValue=<rgba 1 1 1 0.25>}
  NamedRerouteDeclaration(None=LinearInterpolate(A=LinearInterpolate(A=@n55, B=@n57, Alpha=@n58.R), B=Multiply(A=@n59.RGB, B=@n59.RGB), Alpha=@n58.G)) {name=Side_Scale_Mouth, nodeColor=<rgba 0.4062 0 1 1>}
  NamedRerouteDeclaration(None=TextureObjectParameter() {group=Wrap_CactusRocker, parameterName=CaRocker__Text_Msk_Mnth, texture=&T_BlacksRGB}) {name=Text_Mouth_Msk, nodeColor=<rgba 0.6406 1 0 1>}
  n60 = NamedRerouteUsage()
  n61 = Constant3Vector() {constant=<rgba 1 1 -1 0>}
  n62 = MaterialFunctionCall(Color=Add(A=Multiply(A=@n60, B=@n61), B=Constant3Vector() {constant=<rgba 0 0 1 0>})) {materialFunction=&SplitComponents}
  n63 = AppendVector(A=@n62.G, B=@n62.G)
  n64 = AppendVector(A=@n62.R, B=@n62.R)
  n65 = MaterialFunctionCall(Float4=NamedRerouteUsage()) {materialFunction=&BreakOutFloat4Components}
  n66 = LinearInterpolate(
   A=LinearInterpolate(A=LinearInterpolate(A=@n63, B=@n64, Alpha=@n65.R), B=LinearInterpolate(A=@n63, B=@n64, Alpha=@n65.A), Alpha=@n65.B)
   B=AppendVector(A=@n62.R, B=@n62.R)
   Alpha=ScalarParameter() {group=Wrap_CactusRocker, parameterName=CaRocker_Is_Roll, sortPriority=0}
  )
  n67 = ComponentMask(None=NamedRerouteUsage()) {g=true, r=true}
  n68 = MaterialFunctionCall(
   UVs=MaterialFunctionCall(UVs=Add(A=@n66, B=@n67) {constB=0}, Rotation Center=Constant2Vector() {g=0.5, r=0.5}, "Rotation Angle (0-1)"=NamedRerouteUsage()) {materialFunction=&CustomRotator}.Rotated Values
   Texture Scale=ComponentMask(None=NamedRerouteUsage()) {g=true, r=true}
  ) {materialFunction=&ScaleUVsByCenter}
  n69 = TextureSample(UVs=@n68.UVs, Tex=NamedRerouteUsage()) {samplerSource=SSM_Clamp_WorldGroupSettings}
  NamedRerouteDeclaration(None=Multiply(A=@n69.RGBA, B=@n68.0-1 mask)) {name=Mouth_Msk, nodeColor=<rgba 0.9922 0 1 1>}
  n70 = VectorParameter() {group=Wrap_CactusRocker, parameterName=CaRocker_Color_EyeBrow, defaultValue=<rgba 0.855 0.0194 0.1559 1>}
  n71 = ComponentMask(None=@n68.UVs) {r=true}
  n72 = Multiply(
   A=LinearInterpolate(
    A=LinearInterpolate(B=@n70.RGB, Alpha=@n71)
    B=VectorParameter() {group=Wrap_CactusRocker, parameterName=CaRocker_Color_Teeth, defaultValue=<rgba 1 1 1 1>}.RGB
    Alpha=@n69.A
   )
   B=@n69.G
  )
  n73 = Multiply(A=VectorParameter() {group=Wrap_CactusRocker, parameterName=CaRocker_Color_Eye, defaultValue=<rgba 1 1 1 1>}.RGB, B=@n69.B)
  NamedRerouteDeclaration(None=Multiply(A=Add(A=@n72, B=@n73), B=@n68.0-1 mask)) {name=Emissive_Mouth, nodeColor=<rgba 1 0 0.4687 1>}
  NamedRerouteDeclaration(None=Multiply(A=Add(A=@n69.G, B=@n69.G), B=@n68.0-1 mask)) {name=Emissive_Mouth_Msk, nodeColor=<rgba 1 0 0.4688 1>}
  n74 = TextureObjectParameter() {
   group=Wrap_CactusRocker
   parameterName=CaRocker__Text_Mouth_N
   samplerType=SAMPLERTYPE_Normal
   texture=&FlatNormalBlue
  }
  NamedRerouteDeclaration(None=@n74) {name=Text_Mouth_Nrml, nodeColor=<rgba 0.6406 1 0 1>}
  n75 = TextureSample(UVs=@n10.UVs, Tex=NamedRerouteUsage()) {samplerType=SAMPLERTYPE_Normal}
  n76 = LinearInterpolate(A=@n75.RGB, Alpha=@n37) {constB=0}
  n77 = NamedRerouteUsage()
  n78 = Add(A=Multiply(A=@n39, B=@n15), B=@n8)
  n79 = ScalarParameter() {group=Wrap_CactusRocker, parameterName=CaRocker_Roughness_High}
  n80 = ScalarParameter() {defaultValue=1, group=Wrap_CactusRocker, parameterName=CaRocker_Roughness_Low}
  n81 = Saturate(None=Add(A=@n38, B=NamedRerouteUsage()))
  n82 = LinearInterpolate(A=@n10.Raw_Normal, Alpha=@n37) {constB=0}
  n83 = SetMaterialAttributes!(
   MaterialAttributes=NamedRerouteUsage()
   BaseColor=Add(A=LinearInterpolate(A=@n76, Alpha=@n77) {constB=0}, B=@n78)
   Roughness=LinearInterpolate(
    A=LinearInterpolate(A=@n79, B=@n80, Alpha=@n75.A)
    B=ScalarParameter() {defaultValue=0.3, group=Wrap_CactusRocker, parameterName=CaRocker_Roughness_Emissive}
    Alpha=@n81
   )
   Specular=LinearInterpolate(
    A=ScalarParameter() {defaultValue=0.5, group=Wrap_CactusRocker, parameterName=CaRocker_Specular}
    B=ScalarParameter() {defaultValue=0.5, group=Wrap_CactusRocker, parameterName=CaRocker_Specular_Emissive}
    Alpha=@n81
   )
   Metallic=LinearInterpolate(
    A=ScalarParameter() {group=Wrap_CactusRocker, parameterName=CaRocker_Metallic}
    B=ScalarParameter() {defaultValue=0.5, group=Wrap_CactusRocker, parameterName=CaRocker_Metallic_Emissive}
    Alpha=@n39
   )
   Normal=Add(A=LinearInterpolate(A=@n82, Alpha=@n18) {constB=0}, B=NamedRerouteUsage())
   EmissiveColor=@n78
  ) {attributeSetTypes [
    #2
    #3
    #5
    #4
    #1
    #6
   ]}
  NamedRerouteDeclaration(None=@n83) {name=Material_Back, nodeColor=<rgba 0 0.3906 1 1>}
  n84 = TextureObjectParameter() {
   group=Wrap_CactusRocker
   parameterName=CaRocker__Text_N
   samplerType=SAMPLERTYPE_Normal
   texture=&FlatNormalBlue
  }
  NamedRerouteDeclaration(None=@n84) {name=Text_Nrml, nodeColor=<rgba 0.6406 1 0 1>}
  NamedRerouteDeclaration(None=@n10.UVs) {name=UVs, nodeColor=<rgba 0.5469 1 0 1>}
  n85 = NamedRerouteUsage()
  n86 = MaterialFunctionCall(Float4=NamedRerouteUsage()) {materialFunction=&BreakOutFloat4Components}
  NamedRerouteDeclaration(None=Multiply(
    A=TextureSample(UVs=@n68.UVs, Tex=@n85) {samplerSource=SSM_Clamp_WorldGroupSettings}.A
    B=LinearInterpolate(B=@n68.0-1 mask, Alpha=@n86.A) {constA=1}
   )) {name=Mouth_Msk2, nodeColor=<rgba 0.9922 0 1 1>}
  n87 = Multiply(A=Time(), B=ScalarParameter() {defaultValue=0.05, group=Wrap_CactusRocker_Audio, parameterName=CaRocker_Color_Speed}) {constB=0.5}
  n88 = CollectionParameter() {collection=&FortniteMaterialParameters, parameterName=EventAudio_Amplitude}
  n89 = CollectionParameter() {collection=&FortniteMaterialParameters, parameterName=EventAudio_Spectrum}
  n90 = MaterialFunctionCall(Float4=@n89) {materialFunction=&BreakOutFloat4Components}
  n91 = Add(A=@n90.R, B=@n90.R)
  n92 = Add(A=@n90.B, B=@n90.B)
  n93 = Add(A=@n88, B=Divide(A=Add(A=@n91, B=@n92)) {constB=4})
  n94 = ScalarParameter() {defaultValue=1, group=Wrap_CactusRocker_Audio, parameterName=CaRocker_EventAudioMultiplier}
  n95 = NamedRerouteDeclaration(None=Min(A=Multiply(A=@n93, B=@n94)) {constB=30, desc=prevent crazy values}) {name=AudioSpectrum, nodeColor=<rgba 1 0.8203 0 1>}
  n96 = Multiply(A=ScalarParameter() {defaultValue=0.7, group=Wrap_CactusRocker_Audio, parameterName=CaRocker_Color_SpeedPulseStrength}, B=@n95) {constB=0.7}
  NamedRerouteDeclaration(None=If(A=@n95, "A > B"=Add(A=@n87, B=@n96) {constB=0.2}, "A < B"=@n87) {constB=0.1}) {name=Audio_Speed, nodeColor=<rgba 0 1 0.2734 1>}
  n97 = VectorParameter() {group=Wrap_CactusRocker_UVs, parameterName=CaRocker_Offset_Vehicle, defaultValue=<rgba 0 0 0 1>}
  n98 = VectorParameter() {group=Wrap_CactusRocker_UVs, parameterName=CaRocker_Offset_VehicleSide, defaultValue=<rgba 0 0 0 1>}
  n99 = MaterialFunctionCall(Float4=NamedRerouteUsage()) {materialFunction=&BreakOutFloat4Components}
  n100 = LinearInterpolate(
   A=LinearInterpolate(
    A=VectorParameter() {group=Wrap_CactusRocker_UVs, parameterName=CaRocker_Offset_Weapon}.RGB
    B=LinearInterpolate(A=@n97.RGB, B=@n98.RGB, Alpha=@n99.G)
    Alpha=@n99.R
   )
   B=LinearInterpolate(
    A=VectorParameter() {group=Wrap_CactusRocker_UVs, parameterName=CaRocker_Offset_Mech, defaultValue=<rgba 0 0 0 1>}.RGB
    B=VectorParameter() {group=Wrap_CactusRocker_UVs, parameterName=CaRocker_Offset_Characte, defaultValue=<rgba 0 0 0 1>}.RGB
    Alpha=@n99.A
   )
   Alpha=@n99.B
  )
  n101 = VectorParameter() {group=Wrap_CactusRocker_UVs, parameterName=CaRocker_Offset_Backpack, defaultValue=<rgba 0 0 0 1>}
  n102 = MaterialFunctionCall(Float2=NamedRerouteUsage()) {materialFunction=&BreakOutFloat2Components}
  NamedRerouteDeclaration(None=LinearInterpolate(
    A=LinearInterpolate(A=@n100, B=@n101.RGB, Alpha=@n102.R)
    B=VectorParameter() {group=Wrap_CactusRocker_UVs, parameterName=CaRocker_Offset_Pickaxe, defaultValue=<rgba 0 0 0 1>}.RGB
    Alpha=@n102.G
   )) {name=Side_offset, nodeColor=<rgba 0 0.625 1 1>}
  n103 = ScalarParameter() {group=Wrap_CactusRocker_UVs, parameterName=CaRocker_Rotation_Weapon}
  n104 = MaterialFunctionCall(Float4=NamedRerouteUsage()) {materialFunction=&BreakOutFloat4Components}
  n105 = LinearInterpolate(
   A=ScalarParameter() {group=Wrap_CactusRocker_UVs, parameterName=CaRocker_Rotation_Vehicle}
   B=ScalarParameter() {group=Wrap_CactusRocker_UVs, parameterName=CaRocker_Rotation_VehicleSide}
   Alpha=@n104.G
  )
  n106 = ScalarParameter() {group=Wrap_CactusRocker_UVs, parameterName=CaRocker_Rotation_Mech}
  n107 = ScalarParameter() {group=Wrap_CactusRocker_UVs, parameterName=CaRocker_Rotation_Character}
  NamedRerouteDeclaration(None=LinearInterpolate(A=LinearInterpolate(A=@n103, B=@n105, Alpha=@n104.R), B=LinearInterpolate(A=@n106, B=@n107, Alpha=@n104.A), Alpha=@n104.B)) {name=Side_Rotation, nodeColor=<rgba 1 0.7266 0 1>}
  n108 = VectorParameter() {group=Wrap_CactusRocker_UVs, parameterName=CaRocker_Mouth_Offset_Weapon, defaultValue=<rgba -0.38 -0.025 0 0>}
  n109 = VectorParameter() {group=Wrap_CactusRocker_UVs, parameterName=CaRocker_Mouth_Offset_Vehicle, defaultValue=<rgba 0 0 0 1>}
  n110 = VectorParameter() {group=Wrap_CactusRocker_UVs, parameterName=CaRocker_Mouth_Offset_VehicleSide, defaultValue=<rgba 0 0 0 1>}
  n111 = MaterialFunctionCall(Float4=NamedRerouteUsage()) {materialFunction=&BreakOutFloat4Components}
  n112 = LinearInterpolate(
   A=LinearInterpolate(A=@n108.RGB, B=LinearInterpolate(A=@n109.RGB, B=@n110.RGB, Alpha=@n111.G), Alpha=@n111.R)
   B=LinearInterpolate(
    A=VectorParameter() {group=Wrap_CactusRocker_UVs, parameterName=CaRocker_Mouth_Offset_Mech, defaultValue=<rgba 0 0 0 1>}.RGB
    B=VectorParameter() {group=Wrap_CactusRocker_UVs, parameterName=CaRocker_Mouth_Offset_Characte, defaultValue=<rgba 0 0 0 1>}.RGB
    Alpha=@n111.A
   )
   Alpha=@n111.B
  )
  n113 = VectorParameter() {group=Wrap_CactusRocker_UVs, parameterName=CaRocker_Mouth_Offset_Backpack, defaultValue=<rgba 0 0 0 1>}
  n114 = MaterialFunctionCall(Float2=NamedRerouteUsage()) {materialFunction=&BreakOutFloat2Components}
  NamedRerouteDeclaration(None=LinearInterpolate(
    A=LinearInterpolate(A=@n112, B=@n113.RGB, Alpha=@n114.R)
    B=VectorParameter() {group=Wrap_CactusRocker_UVs, parameterName=CaRocker_Mouth_Offset_Pickaxe, defaultValue=<rgba 0 0 0 1>}.RGB
    Alpha=@n114.G
   )) {name=Side_offset_Mouth, nodeColor=<rgba 0 0.625 1 1>}
  NamedRerouteDeclaration(None=FunctionInput(Preview=Constant4Vector().RGBA) {bUsePreviewValueAsDefault=true, inputName=IsState, inputType=FunctionInput_Vector4, sortPriority=4}) {name=IsState, nodeColor=<rgba 0.1497 1 0.1633 1>}
  NamedRerouteDeclaration(None=FunctionInput(Preview=Constant2Vector()) {bUsePreviewValueAsDefault=true, inputName=IsState_B, inputType=FunctionInput_Vector2, sortPriority=4}) {name=IsState_B, nodeColor=<rgba 0.1497 1 0.1633 1>}
  n115 = PreSkinnedLocalBounds()
  n116 = Subtract(A=VertexInterpolator(VS=PreSkinnedPosition()).PS, B=@n115.Min)
  n117 = MaterialFunctionCall(Color=Subtract(A=@n115.Max, B=@n115.Max)) {materialFunction=&SplitComponents}
  n118 = Max(A=Max(A=@n117.R, B=@n117.R), B=@n117.B)
  NamedRerouteDeclaration(None=FunctionInput(Preview=Divide(A=@n116, B=@n118)) {bUsePreviewValueAsDefault=true, inputName=Calcated_Bounds, sortPriority=11}) {name=Caclcated_Bounds, nodeColor=<rgba 0 0.9297 1 1>}
  n119 = ScalarParameter() {defaultValue=1, group=Wrap_CactusRocker_UVs, parameterName=CaRocker_Scale_Vehicle}
  n120 = ScalarParameter() {defaultValue=1, group=Wrap_CactusRocker_UVs, parameterName=CaRocker_Scale_Vehicle_Side}
  n121 = MaterialFunctionCall(Float4=NamedRerouteUsage()) {materialFunction=&BreakOutFloat4Components}
  n122 = LinearInterpolate(
   A=LinearInterpolate(
    A=ScalarParameter() {defaultValue=1, group=Wrap_CactusRocker_UVs, parameterName=CaRocker_Scale_Weapon}
    B=LinearInterpolate(A=@n119, B=@n120, Alpha=@n121.G)
    Alpha=@n121.R
   )
   B=LinearInterpolate(
    A=ScalarParameter() {defaultValue=1, group=Wrap_CactusRocker_UVs, parameterName=CaRocker_Scale_Mech}
    B=ScalarParameter() {defaultValue=1, group=Wrap_CactusRocker_UVs, parameterName=CaRocker_Scale_Character}
    Alpha=@n121.A
   )
   Alpha=@n121.B
  )
  n123 = ScalarParameter() {defaultValue=1, group=Wrap_CactusRocker_UVs, parameterName=CaRocker_Scale_Backpack}
  n124 = MaterialFunctionCall(Float2=NamedRerouteUsage()) {materialFunction=&BreakOutFloat2Components}
  NamedRerouteDeclaration(None=LinearInterpolate(
    A=LinearInterpolate(A=@n122, B=@n123, Alpha=@n124.R)
    B=ScalarParameter() {defaultValue=1, group=Wrap_CactusRocker_UVs, parameterName=CaRocker_Scale_Pickaxe}
    Alpha=@n124.G
   )) {name=Side_Scale, nodeColor=<rgba 0.4062 0 1 1>}
  n125 = PreSkinnedNormal()
  NamedRerouteDeclaration(None=FunctionInput(Preview=VertexInterpolator(VS=@n125).PS) {bUsePreviewValueAsDefault=true, inputName=Pre-Skinned Local Normal, sortPriority=10}) {name=PreSkin_Normal, nodeColor=<rgba 0.8281 1 0 1>}
  NamedRerouteDeclaration(None=@n7.Normal) {name=Root_Normal, nodeColor=<rgba 0.4062 0 1 1>}
  Power(Base=@n22)
  Saturate(None=@n2)
  ScalarParameter() {group=Wrap_CactusRocker, parameterName=CaRocker_Is_Roll, sortPriority=0}
 ]}
