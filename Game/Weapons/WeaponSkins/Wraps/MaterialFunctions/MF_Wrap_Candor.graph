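# Material function for the Candor cosmetic wrap, switched over the incoming material and blended through primary and secondary masks, with an accent material.
# Two side-projected, warped texture layers take offset, scale and rotation per item type (weapon, vehicle, mech, character, backpack, pickaxe).
# They feed a hologram effect; flicker, scanline and rainbow-glow signals are also built.
# The base material combines a bounds-based colour gradient with emissive glow scaled by time of day.
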MaterialFunction /Game/Weapons/WeaponSkins/Wraps/MaterialFunctions/MF_Wrap_Candor
graph {nodes [
  n1 = NamedRerouteUsage()
  Add(A=Multiply(A=@n1, B=Constant2Vector() {g=1, r=1.5}), B=Constant2Vector() {r=-0.25})
  n2 = ComponentMask(None=NamedRerouteUsage()) {r=true}
  n3 = ScalarParameter() {defaultValue=0.02, group=Wrap_Candor, parameterName=Cand_Warp_Strength}
  n4 = Constant() {r=0.5}
  n5 = MaterialFunctionCall(Float2=@n1) {materialFunction=&BreakOutFloat2Components}
  n6 = Constant() {r=0.25}
  n7 = SphereMask(A=@n4, B=@n5.R, Hardness=@n6) {attenuationRadius=0.5}
  n8 = SphereMask(A=@n5.G, B=@n4, Hardness=@n6) {attenuationRadius=0.5}
  n9 = NamedRerouteUsage()
  n10 = MaterialFunctionCall!(
   Normal_Map
   Offset Rt
   Pivot
   Caculated_Bounds=NamedRerouteUsage()
   Pre_Skinned_Normals=NamedRerouteUsage()
   is_State=NamedRerouteUsage()
   UV_Rotation=NamedRerouteUsage()
   Uvs_Offset=ComponentMask(None=@n9) {b=true, g=true, r=true}
   Uvs_TileRate=ComponentMask(None=@n9) {a=true}
   Uvs_Offset_Pivot
   For_Safe_Zones=StaticBool() {value=true}
   Rotation_Comp
   Clamp Y
   Clamp X
   Clamp Padding_Comp
   No_Rotation
  ) {materialFunction=&/MF_Projection_Side_Normal_V2}
  n11 = NamedRerouteDeclaration(None=@n10.UVs) {name=UVs_B, nodeColor=<rgba 1 0.4219 0 1>}
  n12 = Constant2Vector() {g=0.5, r=0.5}
  n13 = Add(
   A=Multiply(A=Multiply(A=@n2, B=@n3) {constA=0.25, constB=0.05}, B=Multiply(A=@n7, B=@n8))
   B=Add(A=Multiply(A=@n11, B=@n12), B=Constant2Vector() {g=0.5})
  )
  BumpOffset(Coordinate=@n13, Height=ScalarParameter() {group=Wrap_Candor_UVs, parameterName=Cand_Bump_Offset_B})
  BumpOffset(Coordinate=@n13, Height=ScalarParameter() {group=Wrap_Candor_UVs, parameterName=Cand_Bump_Offset_A})
  n14 = ScalarParameter() {defaultValue=0.75, group=Wrap_Candor_Glitch, parameterName=Cand_Glitch_FlickerFrequency}
  n15 = NamedRerouteUsage()
  n16 = NamedRerouteUsage()
  n17 = LinearInterpolate(
   A=@n16
   B=Add(A=@n16, B=ScalarParameter() {defaultValue=0.5, group=Wrap_Candor_Glitch, parameterName=Cand_Glitch_Offset_time}) {constB=0.5}
   Alpha=NamedRerouteUsage()
  )
  n18 = Multiply(A=ScalarParameter() {defaultValue=1, group=Wrap_Candor_Glitch, parameterName=Cand_Glitch_FlickerSpeed}, B=@n17)
  n19 = Add(
   A=Multiply(
    A=MaterialFunctionCall(Float2=@n15) {materialFunction=&BreakOutFloat2Components}.G
    B=ScalarParameter() {defaultValue=0.25, group=Wrap_Candor_Glitch, parameterName=Cand_Glitch_FlickerUVScale}
   ) {constB=0.25}
   B=@n18
  )
  n20 = Add(A=Multiply(B=@n14) {constA=0.9, constB=0.9}, B=Sine(None=@n19) {period=1.25}) {constB=0.9}
  Ceil(None=@n20)
  ComponentMask() {g=true, r=true}
  Constant()
  n21 = TextureSample() {samplerType=SAMPLERTYPE_Masks, texture=/Game/Weapons/FORT_Rifles/Textures/T_Scar_CM}
  n22 = FunctionInput(Preview=@n21.R) {bUsePreviewValueAsDefault=true, inputName=Mask Primary, inputType=FunctionInput_Scalar, sortPriority=24, previewValue=<1 0 0 1>}
  n23 = FunctionInput(Preview=@n21.B) {bUsePreviewValueAsDefault=true, inputName=Mask Secondary, inputType=FunctionInput_Scalar, sortPriority=24, previewValue=<1 0 0 1>}
  n24 = BlendMaterialAttributes(A=NamedRerouteUsage(), B=NamedRerouteUsage(), Alpha=Saturate(None=Add(A=@n22, B=@n23))) {vertexAttributeBlendType=UseA}
  n25 = NamedRerouteUsage()
  n26 = NamedRerouteDeclaration(None=@n23) {name=Msk_Accent, nodeColor=<rgba 0 1 0.7656 1>}
  n27 = TextureSample() {samplerType=SAMPLERTYPE_Normal, texture=/Game/Weapons/FORT_Rifles/T_SCAR_N}
  n28 = Constant3Vector() {constant=<rgba 1 0 0 0>}
  n29 = GetMaterialAttributes(None=FunctionInput(Preview=SetMaterialAttributes!(MaterialAttributes, Normal=@n27.RGB, BaseColor=@n28) {attributeSetTypes [#1, #2]}) {inputName=Material, inputType=FunctionInput_MaterialAttributes}) {attributeGetTypes [
    #2
    #3
    #1
    #4
   ]}
  FunctionOutput(None=StaticSwitchParameter(
    True=BlendMaterialAttributes(A=@n24, B=@n25, Alpha=@n26) {vertexAttributeBlendType=UseA}
    False=NamedRerouteDeclaration(None=@n29.MaterialAttributes) {name=Root_Material, nodeColor=<rgba 0.7812 1 0 1>}
   ) {group=Customization_Wrap, parameterName=Wrap_Candor})
  n30 = ComponentMask(None=NamedRerouteUsage()) {r=true}
  n31 = Constant() {r=0.5}
  n32 = Add(A=Multiply(A=NamedRerouteUsage(), B=Constant2Vector() {g=1.4, r=1}), B=Constant2Vector() {g=-0.2})
  n33 = MaterialFunctionCall(Float2=@n32) {materialFunction=&BreakOutFloat2Components}
  n34 = Constant() {r=0.25}
  n35 = SphereMask(A=@n31, B=@n33.R, Hardness=@n34) {attenuationRadius=0.5}
  n36 = SphereMask(A=@n33.G, B=@n31, Hardness=@n34) {attenuationRadius=0.5}
  n37 = NamedRerouteUsage()
  n38 = NamedRerouteUsage()
  n39 = MaterialFunctionCall(
   Caculated_Bounds=@n37
   Pre_Skinned_Normals=NamedRerouteUsage()
   is_State=NamedRerouteUsage()
   UV_Rotation=NamedRerouteUsage()
   Uvs_Offset=ComponentMask(None=@n38) {b=true, g=true, r=true}
   Uvs_TileRate=Multiply(A=ComponentMask(None=@n38) {a=true}, B=Constant3Vector() {constant=<rgba 1 0.5 0 0>})
   For_Safe_Zones=StaticBool() {value=true}
  ) {materialFunction=&MF_Projection_Side_Normal_V2}
  n40 = Add(
   A=Multiply(A=Multiply(A=@n30, B=@n3) {constA=0.25, constB=0.25}, B=Multiply(A=@n35, B=@n36))
   B=Multiply(A=NamedRerouteDeclaration(None=@n39.UVs) {name=UVs, nodeColor=<rgba 1 0.4219 0 1>}, B=Constant2Vector() {g=0.5, r=1})
  )
  MaterialFunctionCall(
   UVs=@n40
   Texture Scale=ComponentMask(None=VectorParameter() {group=Wrap_Candor_UVs, parameterName=Cand_GlowA_Size, defaultValue=<rgba 1 1 0 1>}.RGB) {g=true, r=true}
  ) {materialFunction=&ScaleUVsByCenter}
  n41 = MaterialFunctionCall() {materialFunction=/Engine/Functions/Engine_MaterialFunctions02/WorldPositionOffset/ObjectScale}
  n42 = Max(A=@n41.Scale Y, B=@n41.Scale Y)
  Max(A=@n41.Scale X, B=@n42)
  Multiply()
  Multiply(A=MaterialFunctionCall(Value_Morning=Constant() {r=0.5}, Value_Day=Constant() {r=1}, Value_Evening=Constant() {r=0.5}, Value_Night=Constant() {r=0.25}) {materialFunction=&Scalar_Time_of_Day}.Result)
  Multiply(A=@n41.Scale X, B=@n42)
  n43 = NamedRerouteUsage()
  Multiply(A=@n43, B=ScalarParameter() {defaultValue=0.25, group=Wrap_Candor_Glitch, parameterName=Cand_Holo_Scan_Size})
  n44 = NamedRerouteUsage()
  n45 = Multiply(
   A=MaterialFunctionCall(Float2=@n44) {materialFunction=&BreakOutFloat2Components}.G
   B=ScalarParameter() {defaultValue=250, group=Base_HoloFX_Scanline, parameterName=Scanline_GlitchDensity}
  )
  n46 = Multiply(A=Add(A=Sine(None=@n45), B=Constant() {r=-0.3}), B=Constant() {r=0.5})
  n47 = NamedRerouteUsage()
  n48 = Sine(None=Multiply(A=@n47, B=ScalarParameter() {defaultValue=1, group=Base_HoloFX_Scanline, parameterName=Scanline_Breath_Frequency}))
  n49 = Multiply(
   A=ScalarParameter() {defaultValue=0.2, group=Wrap_Candor_ScanLines, parameterName=Can_Scanline_BaseOpacity}
   B=MaterialFunctionCall(
    Input=ConstantBiasScale(None=@n48)
    Input Low=Constant()
    Input High=Constant() {r=1}
    Target Low=ScalarParameter() {defaultValue=0.8, group=Base_HoloFX_Scanline, parameterName=Scanline_Breath_DarkPercent}
    Target High=ScalarParameter() {defaultValue=1, group=Base_HoloFX_Scanline, parameterName=Scanline_Breath_BrightPercent}
   ) {materialFunction=&RemapValueRange}.Result
  )
  n50 = Constant()
  n51 = ScalarParameter() {defaultValue=0.01, group=Wrap_Candor_ScanLines, parameterName=Scanline_FlowSpeed}
  n52 = Add(A=@n44, B=Multiply(A=AppendVector(A=@n50, B=@n51), B=@n47))
  n53 = Frac(None=Multiply(
    A=MaterialFunctionCall(Float2=@n52) {materialFunction=&BreakOutFloat2Components}.G
    B=ScalarParameter() {defaultValue=250, group=Wrap_Candor_ScanLines, parameterName=Can_Scanline_Density}
   ))
  n54 = OneMinus(None=@n53)
  n55 = ScalarParameter() {defaultValue=-0.5, group=Wrap_Candor_ScanLines, parameterName=Can_Scanline_ThicknessOffset}
  n56 = Add(A=Multiply(A=Min(A=@n54, B=@n53)) {constB=2}, B=@n55)
  n57 = Divide(B=Add(A=@n55))
  n58 = Add(A=ScalarParameter() {defaultValue=3, group=Wrap_Candor_ScanLines, parameterName=Can_Scanline_Sharpness})
  n59 = Saturate(None=Multiply(A=Multiply(A=@n56, B=@n57), B=Clamp(None=@n58) {clampMode=CMODE_ClampMin, minDefault=1}))
  Multiply(A=Saturate(None=Saturate(None=@n46)), B=Saturate(None=LinearInterpolate(A=@n49, Alpha=@n59)))
  n60 = TextureSample(UVs=@n40, Tex=NamedRerouteUsage())
  Multiply(A=@n60.RGB, B=@n60.RGB)
  n61 = VectorParameter() {group=Wrap_Candor_Accent, parameterName=Cand_Accent_BaseColor, defaultValue=<rgba 0.1356 0.2159 0.6584 1>}
  n62 = SetMaterialAttributes!(
   MaterialAttributes=NamedRerouteUsage()
   BaseColor=@n61.RGB
   Roughness=ScalarParameter() {defaultValue=0.3, group=Wrap_Candor_Accent, parameterName=Cand_Accent_Roughness}
   Specular=ScalarParameter() {defaultValue=0.5, group=Wrap_Candor_Accent, parameterName=Cand_Accent_Spec}
   Metallic=ScalarParameter() {defaultValue=0.5, group=Wrap_Candor_Accent, parameterName=Cand_Accent_Metalic}
   Normal
   EmissiveColor
  ) {attributeSetTypes [
    #2
    #3
    #5
    #4
    #1
    #6
   ]}
  NamedRerouteDeclaration(None=@n62) {name=Mat_Accent, nodeColor=<rgba 0 0.3906 1 1>}
  n63 = FunctionInput(Preview=@n21.G) {bUsePreviewValueAsDefault=true, inputName=Mask Crevise, inputType=FunctionInput_Scalar, sortPriority=24, previewValue=<1 0 0 1>}
  NamedRerouteDeclaration(None=@n63) {name=Msk_Curvature, nodeColor=<rgba 0 1 0.7422 1>}
  n64 = Time() {bOverride_Period=true, period=6000}
  n65 = ScalarParameter() {defaultValue=1, group=Wrap_Candor_Glitch, parameterName=Cand_Holo_Speed}
  n66 = NamedRerouteUsage()
  n67 = NamedRerouteUsage()
  n68 = TextureSample(UVs=@n13, Tex=NamedRerouteUsage())
  n69 = Add(A=Multiply(A=@n60.RGB, B=Multiply(A=@n66, B=@n60.A) {constA=1}), B=Multiply(A=Multiply(A=@n67, B=@n68.A), B=@n68.RGB))
  n70 = Add(A=Multiply(A=@n60.A, B=@n66), B=Multiply(A=@n68.A, B=@n67))
  n71 = MaterialFunctionCall!(
   Time=NamedRerouteDeclaration(None=Multiply(A=@n64, B=@n65) {constB=0}) {name=Time, nodeColor=<rgba 0.4427 0 0.0117 1>}
   MeshScaleIn
   UVForScanline=@n43
   CustomFlicker=NamedRerouteUsage()
   UseCustomFlicker=Constant() {r=1}
   UVForStroke=Multiply(A=@n43, B=ScalarParameter() {defaultValue=2, group=Wrap_Candor_Glitch, parameterName=Cand_Holo_Stroke_Size})
   UVForGlitch=@n43
   BaseUVIn
   EmissiveIn=Desaturation(None=Multiply(A=@n69) {constB=0.5}, Fraction=Constant() {r=-0.5})
   OpacityIn=NamedRerouteDeclaration(None=Saturate(None=@n70)) {name=Msk, nodeColor=<rgba 0.6172 1 0 1>}
   WPOIn
  ) {materialFunction=/Game/Tools/3DHB/Shader/Base/MF_3DHB_HologramFX}
  NamedRerouteDeclaration(None=@n71.WpoOut) {name=Uvs_Warp, nodeColor=<rgba 0 1 0.9766 1>}
  n72 = Constant() {r=0.5}
  n73 = MaterialFunctionCall(Float2=@n32) {materialFunction=&BreakOutFloat2Components}
  NamedRerouteDeclaration(None=Multiply(A=SphereMask(A=@n72, B=@n73.R) {attenuationRadius=0.5}, B=SphereMask(A=@n73.G, B=@n72) {attenuationRadius=0.5})) {name=UVs_Msk_A, nodeColor=<rgba 1 0.4219 0 1>}
  n74 = TextureObjectParameter() {group=Wrap_Candor, parameterName=Cond__Tex_N, samplerType=SAMPLERTYPE_Normal, texture=&FlatNormalBlue}
  NamedRerouteDeclaration(None=@n74) {name=Tex_N, nodeColor=<rgba 0.1016 0 1 1>}
  NamedRerouteDeclaration(None=TextureSample(UVs=@n13, Tex=NamedRerouteUsage()) {samplerType=SAMPLERTYPE_Normal}.RGB) {name=Tex_N_B, nodeColor=<rgba 0.2187 1 0 1>}
  NamedRerouteDeclaration(None=TextureSample(UVs=@n40, Tex=NamedRerouteUsage()) {samplerType=SAMPLERTYPE_Normal}.RGB) {name=Tex_N_A, nodeColor=<rgba 0.2188 1 0 1>}
  n75 = MaterialFunctionCall!(
   Bounds=NamedRerouteUsage()
   ForwardAxis=LinearInterpolate(
    A=NamedRerouteUsage()
    B=Constant3Vector() {constant=<rgba 0 1 0 0>}
    Alpha=ScalarParameter() {group=Wrap_Candor, parameterName=Cand_Is_Roll}
   )
   Offset=Constant()
   FallOff=ScalarParameter() {defaultValue=1, group=Wrap_Candor, parameterName=Cand_BaseColor_Falloff}
  ) {materialFunction=&/MF_Wrap_Gradent_Simple}
  n76 = Multiply(
   A=ScalarParameter() {defaultValue=1, group=Wrap_Candor, parameterName=Cand_BaseColor_Brightness}
   B=LinearInterpolate(
    A=VectorParameter() {group=Wrap_Candor, parameterName=Cand_BaseColor_B, defaultValue=<rgba 0.106 0.0367 0.2135 1>}.RGB
    B=VectorParameter() {group=Wrap_Candor, parameterName=Cand_BaseColor_A, defaultValue=<rgba 0.4882 0.148 0.5573 1>}.RGB
    Alpha=Saturate(None=@n75.Clamped_Data)
   )
  )
  n77 = OneMinus(None=NamedRerouteUsage())
  n78 = NamedRerouteUsage()
  n79 = Saturate(None=Add(A=@n77, B=Add(A=@n78) {constB=0.75}) {constB=0})
  n80 = ComponentMask(None=NamedRerouteUsage()) {a=true}
  n81 = Add(A=NamedRerouteUsage(), B=Multiply(A=NamedRerouteUsage(), B=@n80))
  n82 = Add(A=@n77, B=OneMinus(None=Saturate(None=@n81)))
  n83 = Multiply(A=@n71.EmissiveOut, B=NamedRerouteUsage())
  n84 = VectorParameter() {group=Wrap_Candor, parameterName=Cand_Emissive_Glow_A, defaultValue=<rgba 0 0.5 0.4761 1>}
  n85 = LinearInterpolate(
   A=VectorParameter() {group=Wrap_Candor, parameterName=Cand_Emissive_Glow_B, defaultValue=<rgba 0.599 0.0937 0.034 1>}.RGB
   B=@n84.RGB
   Alpha=Power(Base=@n80)
  )
  n86 = Multiply(A=NamedRerouteUsage(), B=Multiply(A=@n80, B=Multiply(A=@n85, B=@n84.A)))
  n87 = Add(A=Multiply(A=Add(A=@n83, B=@n86), B=NamedRerouteUsage()) {constA=1}, B=Multiply(A=NamedRerouteUsage()) {constA=1})
  n88 = NamedRerouteUsage()
  n89 = ScalarParameter() {defaultValue=0.75, group=Wrap_Candor_TOD, parameterName=Cand_TOD_Morn}
  n90 = ScalarParameter() {defaultValue=1, group=Wrap_Candor_TOD, parameterName=Cand_TOD_Day}
  n91 = ScalarParameter() {defaultValue=0.75, group=Wrap_Candor_TOD, parameterName=Cand_TOD_Evening}
  n92 = ScalarParameter() {defaultValue=0.5, group=Wrap_Candor_TOD, parameterName=Cand_TOD_Night}
  n93 = SetMaterialAttributes(
   MaterialAttributes=NamedRerouteUsage()
   BaseColor=Multiply(A=Multiply(A=@n76, B=@n79), B=Saturate(None=@n82)) {constA=1}
   Roughness=ScalarParameter() {defaultValue=0.4, group=Wrap_Candor, parameterName=Cand_Roughness}
   Specular=ScalarParameter() {defaultValue=0.5, group=Wrap_Candor, parameterName=Cand_Spec}
   Metallic=ScalarParameter() {group=Wrap_Candor, parameterName=Cand_Metallic}
   EmissiveColor=Multiply(
    A=Multiply(A=@n87, B=@n88)
    B=MaterialFunctionCall(Value_Morning=@n89, Value_Day=@n90, Value_Evening=@n91, Value_Night=@n92) {materialFunction=&Scalar_Time_of_Day}.Result
   )
  ) {attributeSetTypes [
    #2
    #3
    #5
    #4
    #1
    #6
   ]}
  NamedRerouteDeclaration(None=@n93) {name=Mat_Base, nodeColor=<rgba 0 0.3906 1 1>}
  n94 = VectorParameter() {group=Wrap_Candor_UVs, parameterName=Candor_OffsetA_Scale_Weapon, defaultValue=<rgba -0.3 -0.025 0 0.8>}
  n95 = VectorParameter() {group=Wrap_Candor_UVs, parameterName=Candor_OffsetA_Scale_Vehicle, defaultValue=<rgba 0 0 0 1>}
  n96 = AppendVector(A=@n95.RGB, B=@n95.RGB)
  n97 = VectorParameter() {group=Wrap_Candor_UVs, parameterName=Candor_OffsetA_Scale_VehicleSide, defaultValue=<rgba 0 0 0 1>}
  n98 = AppendVector(A=@n97.RGB, B=@n97.RGB)
  n99 = MaterialFunctionCall(Float4=NamedRerouteUsage()) {materialFunction=&BreakOutFloat4Components}
  n100 = VectorParameter() {group=Wrap_Candor_UVs, parameterName=Candor_OffsetA_Scale_Mech, defaultValue=<rgba 0 0 0 1>}
  n101 = VectorParameter() {group=Wrap_Candor_UVs, parameterName=Candor_OffsetA_Scale_Character, defaultValue=<rgba 0 0 0 1>}
  n102 = LinearInterpolate(
   A=LinearInterpolate(A=AppendVector(A=@n94.RGB, B=@n94.RGB), B=LinearInterpolate(A=@n96, B=@n98, Alpha=@n99.G), Alpha=@n99.R)
   B=LinearInterpolate(A=AppendVector(A=@n100.RGB, B=@n100.RGB), B=AppendVector(A=@n101.RGB, B=@n101.RGB), Alpha=@n99.A)
   Alpha=@n99.B
  )
  n103 = VectorParameter() {group=Wrap_Candor_UVs, parameterName=Candor_OffsetA_Scale_Backpack, defaultValue=<rgba 0 0 0 1>}
  n104 = AppendVector(A=@n103.RGB, B=@n103.RGB)
  n105 = MaterialFunctionCall(Float2=NamedRerouteUsage()) {materialFunction=&BreakOutFloat2Components}
  n106 = VectorParameter() {group=Wrap_Candor_UVs, parameterName=Candor_OffsetA_Scale_Pickaxe, defaultValue=<rgba 0 0 0 1>}
  NamedRerouteDeclaration(None=LinearInterpolate(A=LinearInterpolate(A=@n102, B=@n104, Alpha=@n105.R), B=AppendVector(A=@n106.RGB, B=@n106.RGB), Alpha=@n105.G)) {name=UVs_OffsetA_Scale, nodeColor=<rgba 0 0.625 1 1>}
  n107 = VectorParameter() {group=Wrap_Candor_UVs, parameterName=Candor_OffsetB_Scale_Weapon, defaultValue=<rgba 0.33 -0.08 0 0.25>}
  n108 = VectorParameter() {group=Wrap_Candor_UVs, parameterName=Candor_OffsetB_Scale_Vehicle, defaultValue=<rgba 0 0 0 1>}
  n109 = AppendVector(A=@n108.RGB, B=@n108.RGB)
  n110 = VectorParameter() {group=Wrap_Candor_UVs, parameterName=Candor_OffsetB_Scale_VehicleSide, defaultValue=<rgba 0 0 0 1>}
  n111 = AppendVector(A=@n110.RGB, B=@n110.RGB)
  n112 = MaterialFunctionCall(Float4=NamedRerouteUsage()) {materialFunction=&BreakOutFloat4Components}
  n113 = VectorParameter() {group=Wrap_Candor_UVs, parameterName=Candor_OffsetB_Scale_Mech, defaultValue=<rgba 0 0 0 1>}
  n114 = VectorParameter() {group=Wrap_Candor_UVs, parameterName=Candor_OffsetB_Scale_Character, defaultValue=<rgba 0 0 0 1>}
  n115 = LinearInterpolate(
   A=LinearInterpolate(A=AppendVector(A=@n107.RGB, B=@n107.RGB), B=LinearInterpolate(A=@n109, B=@n111, Alpha=@n112.G), Alpha=@n112.R)
   B=LinearInterpolate(A=AppendVector(A=@n113.RGB, B=@n113.RGB), B=AppendVector(A=@n114.RGB, B=@n114.RGB), Alpha=@n112.A)
   Alpha=@n112.B
  )
  n116 = VectorParameter() {group=Wrap_Candor_UVs, parameterName=Candor_OffsetB_Scale_Backpack, defaultValue=<rgba 0 0 0 1>}
  n117 = AppendVector(A=@n116.RGB, B=@n116.RGB)
  n118 = MaterialFunctionCall(Float2=NamedRerouteUsage()) {materialFunction=&BreakOutFloat2Components}
  n119 = VectorParameter() {group=Wrap_Candor_UVs, parameterName=Candor_OffsetB_Scale_Pickaxe, defaultValue=<rgba 0 0 0 1>}
  NamedRerouteDeclaration(None=LinearInterpolate(A=LinearInterpolate(A=@n115, B=@n117, Alpha=@n118.R), B=AppendVector(A=@n119.RGB, B=@n119.RGB), Alpha=@n118.G)) {name=UVs_OffsetB_Scale, nodeColor=<rgba 0 0.625 1 1>}
  n120 = LinearInterpolate(
   A=LinearInterpolate(A=Constant3Vector() {constant=<rgba 0 0 1 0>}, B=NamedRerouteUsage(), Alpha=NamedRerouteUsage())
   B=NamedRerouteUsage()
   Alpha=NamedRerouteUsage()
  )
  n121 = ScalarParameter() {defaultValue=0.1, group=Wrap_Candor_Rainbow, parameterName=Cand_NormalToEdge_Flat}
  n122 = Power(
   Base=DotProduct(
    A=MaterialFunctionCall(Normal=@n120, Flatness=@n121) {materialFunction=&FlattenNormal}.Result
    B=Constant3Vector() {constant=<rgba 0 0 1 0>}
   )
   Exp=ScalarParameter() {defaultValue=2, group=Wrap_Candor_Rainbow, parameterName=Cand_NormalToEdge_Power}
  )
  NamedRerouteDeclaration(None=Multiply(
    A=MaterialFunctionCall!(
     Time
     "NormalIn(TS)"=@n120
     DirctionIn
     AdditionalDesaturationIn
     AdditionalStrengthIn
     AdditionalDensityIn
     AdditionalFlattenNormalIn
     WPOIn
    ) {materialFunction=/Game/Tools/3DHB/Shader/MasterLib/PostFunctions/MF_3DHB_POST_RainbowColor}.BlendToEmissive
    B=OneMinus(None=@n122)
   )) {name=RainBow_Glow, nodeColor=<rgba 0.9922 0 1 1>}
  NamedRerouteDeclaration(None=@n122) {name=RainBow_Glow_msk, nodeColor=<rgba 0.9922 0 1 1>}
  NamedRerouteDeclaration(None=@n10.Safe_Zone_Mask) {name=UVs_Msk_B, nodeColor=<rgba 1 0.4219 0 1>}
  n123 = Multiply(A=@n59, B=ScalarParameter() {defaultValue=1, group=Base_HoloFX_Scanline, parameterName=Scanline_ToEmissiveScale})
  NamedRerouteDeclaration(None=@n123) {name=Scanline_ToOpacity, nodeColor=<rgba 1 0.75 0 1>}
  n124 = ScalarParameter() {group=Wrap_Candor_UVs, parameterName=Candor_RotationA_Vehicle, sortPriority=150}
  n125 = ScalarParameter() {group=Wrap_Candor_UVs, parameterName=Candor_RotationA_Vehicle_Side, sortPriority=150}
  n126 = MaterialFunctionCall(Float4=NamedRerouteUsage()) {materialFunction=&BreakOutFloat4Components}
  n127 = LinearInterpolate(
   A=LinearInterpolate(
    A=ScalarParameter() {group=Wrap_Candor_UVs, parameterName=Candor_RotationA_Weapon, sortPriority=150}
    B=LinearInterpolate(A=@n124, B=@n125, Alpha=@n126.G)
    Alpha=@n126.R
   )
   B=LinearInterpolate(
    A=ScalarParameter() {group=Wrap_Candor_UVs, parameterName=Candor_RotationA_Mech, sortPriority=150}
    B=ScalarParameter() {group=Wrap_Candor_UVs, parameterName=Candor_RotationA_Character, sortPriority=150}
    Alpha=@n126.A
   )
   Alpha=@n126.B
  )
  n128 = ScalarParameter() {group=Wrap_Candor_UVs, parameterName=Candor_RotationA_Backpack, sortPriority=150}
  n129 = MaterialFunctionCall(Float2=NamedRerouteUsage()) {materialFunction=&BreakOutFloat2Components}
  NamedRerouteDeclaration(None=LinearInterpolate(
    A=LinearInterpolate(A=@n127, B=@n128, Alpha=@n129.R)
    B=ScalarParameter() {group=Wrap_Candor_UVs, parameterName=Candor_RotationA_Pickaxe, sortPriority=150}
    Alpha=@n129.G
   )) {name=UVs_RotationA, nodeColor=<rgba 0 0.7188 1 1>}
  n130 = ScalarParameter() {group=Wrap_Candor_UVs, parameterName=Candor_RotationB_Vehicle, sortPriority=150}
  n131 = ScalarParameter() {group=Wrap_Candor_UVs, parameterName=Candor_RotationB_Vehicle_Side, sortPriority=150}
  n132 = MaterialFunctionCall(Float4=NamedRerouteUsage()) {materialFunction=&BreakOutFloat4Components}
  n133 = LinearInterpolate(
   A=LinearInterpolate(
    A=ScalarParameter() {group=Wrap_Candor_UVs, parameterName=Candor_RotationB_Weapon, sortPriority=150}
    B=LinearInterpolate(A=@n130, B=@n131, Alpha=@n132.G)
    Alpha=@n132.R
   )
   B=LinearInterpolate(
    A=ScalarParameter() {group=Wrap_Candor_UVs, parameterName=Candor_RotationB_Mech, sortPriority=150}
    B=ScalarParameter() {group=Wrap_Candor_UVs, parameterName=Candor_RotationB_Character, sortPriority=150}
    Alpha=@n132.A
   )
   Alpha=@n132.B
  )
  n134 = ScalarParameter() {group=Wrap_Candor_UVs, parameterName=Candor_RotationB_Backpack, sortPriority=150}
  n135 = MaterialFunctionCall(Float2=NamedRerouteUsage()) {materialFunction=&BreakOutFloat2Components}
  NamedRerouteDeclaration(None=LinearInterpolate(
    A=LinearInterpolate(A=@n133, B=@n134, Alpha=@n135.R)
    B=ScalarParameter() {group=Wrap_Candor_UVs, parameterName=Candor_RotationB_Pickaxe, sortPriority=150}
    Alpha=@n135.G
   )) {name=UVs_RotationB, nodeColor=<rgba 0 0.7188 1 1>}
  n136 = Sine(None=@n18) {period=1.6}
  n137 = Add(A=Sine(None=@n19) {period=2.1}, B=Multiply(B=@n14) {constA=0.25})
  n138 = Add(A=Sine(None=@n19) {period=0.72}, B=Multiply(B=@n14) {constA=0.75}) {constB=0.75}
  n139 = Saturate(None=Multiply(A=Multiply(A=@n137, B=@n138), B=Multiply(A=@n138, B=@n20)))
  n140 = LinearInterpolate(
   A=OneMinus(None=ScalarParameter() {defaultValue=0.4, group=Wrap_Candor_Glitch, parameterName=Cand_Glitch_FlickerDetailShakeStrength})
   Alpha=ConstantBiasScale(None=Sine(None=@n17) {period=0.075})
  ) {constA=0.7}
  NamedRerouteDeclaration(None=Multiply(A=Ceil(None=@n136), B=Multiply(A=@n139, B=@n140))) {name=Flicker, nodeColor=<rgba 1 0.2344 0 1>}
  NamedRerouteDeclaration(None=TextureObjectParameter() {group=Wrap_Candor, parameterName=Cond__Tex_Bc, texture=&T_BlacksRGB}) {name=Tex_BC, nodeColor=<rgba 0.1016 0 1 1>}
  n141 = ScalarParameter() {group=Wrap_Candor_UVs, parameterName=Candor_RotationBck_Vehicle, sortPriority=150}
  n142 = ScalarParameter() {group=Wrap_Candor_UVs, parameterName=Candor_RotationBck_Vehicle_Side, sortPriority=150}
  n143 = MaterialFunctionCall(Float4=NamedRerouteUsage()) {materialFunction=&BreakOutFloat4Components}
  n144 = LinearInterpolate(
   A=LinearInterpolate(
    A=ScalarParameter() {group=Wrap_Candor_UVs, parameterName=Candor_RotationBck_Weapon, sortPriority=150}
    B=LinearInterpolate(A=@n141, B=@n142, Alpha=@n143.G)
    Alpha=@n143.R
   )
   B=LinearInterpolate(
    A=ScalarParameter() {group=Wrap_Candor_UVs, parameterName=Candor_RotationBck_Mech, sortPriority=150}
    B=ScalarParameter() {group=Wrap_Candor_UVs, parameterName=Candor_RotationBck_Character, sortPriority=150}
    Alpha=@n143.A
   )
   Alpha=@n143.B
  )
  n145 = ScalarParameter() {group=Wrap_Candor_UVs, parameterName=Candor_RotationBck_Backpack, sortPriority=150}
  n146 = MaterialFunctionCall(Float2=NamedRerouteUsage()) {materialFunction=&BreakOutFloat2Components}
  NamedRerouteDeclaration(None=LinearInterpolate(
    A=LinearInterpolate(A=@n144, B=@n145, Alpha=@n146.R)
    B=ScalarParameter() {group=Wrap_Candor_UVs, parameterName=Candor_RotationBck_Pickaxe, sortPriority=150}
    Alpha=@n146.G
   )) {name=UVs_RotationBck, nodeColor=<rgba 0 0.7188 1 1>}
  NamedRerouteDeclaration(None=@n39.UVs No Offsets/Tiling) {name=UVs_No_1, nodeColor=<rgba 1 0.4219 0 1>}
  NamedRerouteDeclaration(None=TextureObjectParameter() {group=Wrap_Candor, parameterName=Cond__Tex_Msk, texture=&T_BlacksRGB}) {name=Tex_Msk, nodeColor=<rgba 0.1016 0 1 1>}
  n147 = TextureSample(UVs=@n40, Tex=NamedRerouteUsage()) {samplerType=SAMPLERTYPE_LinearColor}
  n148 = NamedRerouteUsage()
  n149 = TextureSample(UVs=@n13, Tex=NamedRerouteUsage()) {samplerType=SAMPLERTYPE_LinearColor}
  n150 = NamedRerouteUsage()
  NamedRerouteDeclaration(None=Add(A=Multiply(A=@n147.RGBA, B=@n148), B=Multiply(A=@n149.RGBA, B=@n150))) {name=Msk_Glow, nodeColor=<rgba 0 1 0.8828 1>}
  NamedRerouteDeclaration(None=MaterialFunctionCall(UVs=NamedRerouteUsage(), Rotation Center=Constant() {r=0.5}, "Rotation Angle (0-1)"=NamedRerouteUsage()) {materialFunction=&CustomRotator}.Rotated Values) {name=UVs_No_1B, nodeColor=<rgba 1 0.4219 0 1>}
  n151 = ScalarParameter() {group=Wrap_Candor_UVs, parameterName=Candor_RotationScan_Vehicle, sortPriority=150}
  n152 = ScalarParameter() {group=Wrap_Candor_UVs, parameterName=Candor_RotationScan_Vehicle_Side, sortPriority=150}
  n153 = MaterialFunctionCall(Float4=NamedRerouteUsage()) {materialFunction=&BreakOutFloat4Components}
  n154 = LinearInterpolate(
   A=LinearInterpolate(
    A=ScalarParameter() {group=Wrap_Candor_UVs, parameterName=Candor_RotationScan_Weapon, sortPriority=150}
    B=LinearInterpolate(A=@n151, B=@n152, Alpha=@n153.G)
    Alpha=@n153.R
   )
   B=LinearInterpolate(
    A=ScalarParameter() {group=Wrap_Candor_UVs, parameterName=Candor_RotationScan_Mech, sortPriority=150}
    B=ScalarParameter() {group=Wrap_Candor_UVs, parameterName=Candor_RotationScan_Character, sortPriority=150}
    Alpha=@n153.A
   )
   Alpha=@n153.B
  )
  n155 = ScalarParameter() {group=Wrap_Candor_UVs, parameterName=Candor_RotationScan_Backpack, sortPriority=150}
  n156 = MaterialFunctionCall(Float2=NamedRerouteUsage()) {materialFunction=&BreakOutFloat2Components}
  NamedRerouteDeclaration(None=LinearInterpolate(
    A=LinearInterpolate(A=@n154, B=@n155, Alpha=@n156.R)
    B=ScalarParameter() {group=Wrap_Candor_UVs, parameterName=Candor_RotationScan_Pickaxe, sortPriority=150}
    Alpha=@n156.G
   )) {name=UVs_RotationScan, nodeColor=<rgba 0 0.7188 1 1>}
  NamedRerouteDeclaration(None=FunctionInput(Preview=Constant3Vector() {constant=<rgba 0 1 0 0>}) {bUsePreviewValueAsDefault=true, inputName=ForwardAxis, sortPriority=100}) {name=FowardAxis, nodeColor=<rgba 0 0.9297 1 1>}
  NamedRerouteDeclaration(None=Add(A=@n123)) {name=Scanline_ToEmissive, nodeColor=<rgba 0 0.7656 1 1>}
  n157 = Transform(None=CameraVectorWS()) {transformSourceType=TRANSFORMSOURCE_World, transformType=TRANSFORM_Local}
  n158 = Multiply(A=Saturate(None=@n41.Scale XYZ))
  NamedRerouteDeclaration(None=Add(A=Multiply(A=@n157, B=@n158), B=@n37)) {name=Bounds_Bump, nodeColor=<rgba 0.1719 0 1 1>}
  NamedRerouteDeclaration(None=FunctionInput(Preview=Constant4Vector().RGBA) {bUsePreviewValueAsDefault=true, inputName=IsState, inputType=FunctionInput_Vector4, sortPriority=4}) {name=IsState, nodeColor=<rgba 0.1497 1 0.1633 1>}
  NamedRerouteDeclaration(None=FunctionInput(Preview=Constant2Vector()) {bUsePreviewValueAsDefault=true, inputName=IsState_B, inputType=FunctionInput_Vector2, sortPriority=4}) {name=IsState_B, nodeColor=<rgba 0.1497 1 0.1633 1>}
  n159 = PreSkinnedLocalBounds()
  n160 = Subtract(A=VertexInterpolator(VS=PreSkinnedPosition()).PS, B=@n159.Min)
  n161 = MaterialFunctionCall(Color=Subtract(A=@n159.Max, B=@n159.Max)) {materialFunction=&SplitComponents}
  n162 = Max(A=Max(A=@n161.R, B=@n161.R), B=@n161.B)
  NamedRerouteDeclaration(None=FunctionInput(Preview=Divide(A=@n160, B=@n162)) {bUsePreviewValueAsDefault=true, inputName=Calcated_Bounds, sortPriority=11}) {name=Caclcated_Bounds, nodeColor=<rgba 0 0.9297 1 1>}
  n163 = MaterialFunctionCall(Color=Abs(None=NamedRerouteUsage())) {materialFunction=&SplitComponents}
  n164 = MaterialFunctionCall(Float4=NamedRerouteUsage()) {materialFunction=&BreakOutFloat4Components}
  n165 = LinearInterpolate(A=@n163.R, B=LinearInterpolate(A=@n163.B, B=@n163.B, Alpha=@n164.G), Alpha=@n164.R)
  NamedRerouteDeclaration(None=MaterialFunctionCall(
    In=LinearInterpolate(A=@n165, B=@n163.G, Alpha=@n164.B)
    Contrast=ScalarParameter() {defaultValue=5, group=Wrap_Candor, parameterName=Candor_Blend_Contrast}
   ) {materialFunction=&CheapContrast}.Result) {name=Tri_Mask, nodeColor=<rgba 1 0.4688 0 1>}
  n166 = PreSkinnedNormal()
  NamedRerouteDeclaration(None=FunctionInput(Preview=VertexInterpolator(VS=@n166).PS) {bUsePreviewValueAsDefault=true, inputName=Pre-Skinned Local Normal, sortPriority=10}) {name=PreSkin_Normal, nodeColor=<rgba 0.8281 1 0 1>}
  NamedRerouteDeclaration(None=@n29.Normal) {name=Root_Normal, nodeColor=<rgba 0.4062 0 1 1>}
  Saturate(None=Add(A=@n80, B=@n80))
 ]}
