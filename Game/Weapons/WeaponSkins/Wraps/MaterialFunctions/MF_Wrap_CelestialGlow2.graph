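# Material function for the CelestialGlow wrap, painting an animated swirling black hole with stars onto weapons, vehicles, mechs and characters.
# A camera-facing virtual plane built from object bounds supplies projected UVs, with per-type offsets, scales and a character safe-zone path.
# Panning noise distorts radial glows and rim, layered with counter-rotating spiral masks, a dark center and fresnel.
# The CelestialGlow static switch blends this emissive material over the input using customization masks.
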MaterialFunction /Game/Weapons/WeaponSkins/Wraps/MaterialFunctions/MF_Wrap_CelestialGlow2
graph {nodes [
  n1 = PreSkinnedLocalBounds()
  n2 = Add(A=@n1.Min, B=FunctionInput() {inputName=Object Bounds Min Adjustment, sortPriority=51}) {constB=0}
  n3 = FunctionInput() {inputName=Object Bounds Max Adjustment, sortPriority=52}
  n4 = Multiply(A=Subtract(A=Add(A=@n1.Max, B=@n3) {constB=0}, B=@n2)) {constB=0.5}
  n5 = FunctionInput(Preview=Constant()) {bUsePreviewValueAsDefault=true, inputName=isMech, inputType=FunctionInput_Scalar}
  n6 = LinearInterpolate(A=TransformPosition(None=Add(A=@n2, B=@n4)) {transformType=TRANSFORMPOSSOURCE_World}, B=ObjectPositionWS(), Alpha=@n5)
  n7 = Normalize(VectorInput=Subtract(A=@n6, B=CameraPositionWS()))
  n8 = CameraVectorWS()
  n9 = Ceil(None=Saturate(None=DotProduct(A=@n7, B=@n8)))
  n10 = VectorParameter() {parameterName=EmColor1, defaultValue=<rgba 0 0 0 1>}
  n11 = VectorParameter() {parameterName=EmColor2, defaultValue=<rgba 0 0 0 1>}
  n12 = ScalarParameter() {defaultValue=0.2, parameterName=DistanceFalloff}
  n13 = VectorParameter() {group=ZeroPoint_UV, parameterName=ZP_Offset_Vehicle}
  n14 = VectorParameter() {group=ZeroPoint_UV, parameterName=ZP_Offset_VehicleSide}
  n15 = Constant()
  n16 = Constant()
  n17 = MaterialFunctionCall(Color=MaterialFunctionCall(
    R=FunctionInput(Preview=@n15) {inputName=isVehicle, inputType=FunctionInput_Scalar, sortPriority=5}
    G=FunctionInput(Preview=@n16) {bUsePreviewValueAsDefault=true, inputName=Side Project Vehicle, inputType=FunctionInput_Scalar}
    B=@n5
   ) {materialFunction=&AppendMany}.RGB) {materialFunction=&SplitComponents}
  n18 = LinearInterpolate(
   A=LinearInterpolate(
    A=VectorParameter() {group=ZeroPoint_UV, parameterName=ZP_Offset_Weapon}.RGB
    B=LinearInterpolate(A=@n13.RGB, B=@n14.RGB, Alpha=@n17.G)
    Alpha=@n17.R
   )
   B=VectorParameter() {group=ZeroPoint_UV, parameterName=ZP_Offset_Mech}.RGB
   Alpha=@n17.B
  )
  n19 = Normalize(VectorInput=CrossProduct(A=@n7, B=VectorParameter() {parameterName=Axis, defaultValue=<rgba 0 0 1 1>}.RGB))
  n20 = MaterialFunctionCall(Float3=PreSkinnedLocalBounds().Half Extents) {materialFunction=&BreakOutFloat3Components}
  n21 = Multiply(
   A=MaterialFunctionCall() {materialFunction=/Engine/Functions/Engine_MaterialFunctions02/WorldPositionOffset/ObjectScale}.Scale X
   B=Max(A=Max(A=@n20.R, B=@n20.R), B=@n20.B)
  )
  n22 = NamedRerouteUsage()
  n23 = LinearInterpolate(
   A=Add(
    A=ComponentMask(None=@n18) {g=true, r=true}
    B=MaterialFunctionCall!(Plane Normal Axis=@n7, Plane X Axis=@n19, Plane Center=@n6, UV Scale=@n21) {materialFunction=/Game/Athena/Environments/Nexus/Materials/VirtualPlaneCoordinatesAbs}.UVs
   )
   B=NamedRerouteUsage()
   Alpha=@n22
  )
  n24 = Multiply(A=@n23, B=ScalarParameter() {defaultValue=1, parameterName=Distortion Tiling1})
  n25 = ScalarParameter() {defaultValue=-0.1, parameterName=panSpeed1}
  n26 = TextureObjectParameter() {
   group=ZeroPoint_Textures
   parameterName=ZP2_DistortionMask
   samplerType=SAMPLERTYPE_LinearColor
   texture=/Game/Characters/Player/Male/Medium/Bodies/M_Med_Soldier_04/Skins/Galaxy/Textures/Twinkle_Star_02__M
  }
  n27 = ChannelMaskParameter(None=TextureSample(UVs=Panner(Coordinate=@n24, Speed=@n25), Tex=@n26) {texture=&T_BlacksRGB}.RGBA) {maskChannel=Green, parameterName=Distortion_Channel, sortPriority=6, defaultValue=<rgba 0 1 0 0>}
  n28 = Panner(
   Coordinate=Multiply(A=@n23, B=ScalarParameter() {defaultValue=1, parameterName=Distortion Tiling2})
   Speed=ScalarParameter() {defaultValue=0.2, parameterName=panSpeed2}
  )
  n29 = Add(
   A=@n27.R
   B=ChannelMaskParameter(None=TextureSample(UVs=@n28, Tex=@n26) {texture=&T_BlacksRGB}.RGBA) {maskChannel=Green, parameterName=Distortion2_Channel, sortPriority=6, defaultValue=<rgba 0 1 0 0>}.R
  )
  n30 = ShadingPathSwitch(Default=@n29, Deferred=@n29, Forward=@n27.R, Mobile=@n27.R)
  n31 = Multiply(A=QualitySwitch(Default=@n30, High=@n30, Medium=@n30), B=ScalarParameter() {defaultValue=-0.1238, parameterName=Distortion})
  n32 = Multiply(
   A=ComponentMask(None=VectorParameter() {parameterName=ScaleSwirlSquash, defaultValue=<rgba 1 0.5 0 1>}.RGB) {g=true, r=true}
   B=ScalarParameter() {defaultValue=1.68, parameterName=ScaleTwirl}
  )
  n33 = Multiply(
   A=ComponentMask(None=VectorParameter() {parameterName=ScaleSwirlSquash_Vehicle, defaultValue=<rgba 0.5 1 0 1>}.RGB) {g=true, r=true}
   B=ScalarParameter() {defaultValue=1.9475, parameterName=ScaleTwirl_Vehicle}
  )
  n34 = ComponentMask(None=VectorParameter() {parameterName=ScaleSwirlSquash_Character, defaultValue=<rgba 0.5 1 0 1>}.RGB) {g=true, r=true}
  n35 = ScalarParameter() {defaultValue=1.9475, parameterName=ScaleTwirl_Character}
  n36 = MaterialFunctionCall(
   UVs=@n23
   Texture Scale=LinearInterpolate(A=LinearInterpolate(A=@n32, B=@n33, Alpha=@n17.R), B=Multiply(A=@n34, B=@n35), Alpha=NamedRerouteUsage())
  ) {materialFunction=&ScaleUVsByCenter}
  n37 = Distance(A=@n36.UVs, B=Constant2Vector() {g=0.5, r=0.5})
  n38 = Fresnel(ExponentIn=ScalarParameter() {defaultValue=0.4967, parameterName=DistortionExp}) {baseReflectFraction=0}
  n39 = ScalarParameter() {defaultValue=0.2744, parameterName=DistortionNormal}
  n40 = Multiply(A=Add(A=@n31, B=@n37), B=OneMinus(None=Multiply(A=@n38, B=@n39)))
  n41 = Clamp(
   None=Power(Base=Divide(A=@n40, B=@n12), Exp=ScalarParameter() {defaultValue=80, parameterName=FresnelEXP})
   Min=ScalarParameter() {defaultValue=0.9, parameterName=max}
  )
  n42 = Multiply(
   A=VectorParameter() {parameterName=CoreGLow, defaultValue=<rgba 5 0.1835 1.9151 1>}.RGB
   B=ScalarParameter() {defaultValue=0.4667, parameterName=Glow -----}
  )
  n43 = Divide(A=@n40, B=ScalarParameter() {defaultValue=1, parameterName=DistanceFalloff2})
  n44 = Clamp(None=Power(Base=OneMinus(None=@n43), Exp=ScalarParameter() {defaultValue=18, parameterName=FresnelEXP3}))
  n45 = Divide(A=@n40, B=ScalarParameter() {defaultValue=1.3, parameterName=DistanceFalloff3})
  n46 = Clamp(None=Power(Base=OneMinus(None=@n45), Exp=ScalarParameter() {defaultValue=3, parameterName=FresnelEXP4}))
  n47 = Add(
   A=Add(A=LinearInterpolate(A=@n10.RGB, B=@n11.RGB, Alpha=@n41), B=Multiply(A=@n42, B=@n44))
   B=Multiply(A=VectorParameter() {parameterName=CoreGLow2, defaultValue=<rgba 0.4688 0.0761 0.2145 1>}.RGB, B=@n46)
  )
  n48 = Constant2Vector() {g=0.5, r=0.5}
  n49 = ScalarParameter() {defaultValue=-0.05, parameterName=ColorCoreTwirl_RotationSpeed}
  n50 = Time()
  n51 = TextureObjectParameter() {
   group=ZeroPoint_Textures
   parameterName=ZP2_Mask
   samplerType=SAMPLERTYPE_Masks
   texture=/Game/Weapons/WeaponSkins/Wraps/Textures/CelestialGlow/CelestialSpiral_output
  }
  n52 = TextureSample(
   UVs=MaterialFunctionCall(UVs=@n36.UVs, Rotation Center=@n48, "Rotation Angle (0-1)"=Multiply(A=@n49, B=@n50)) {materialFunction=&CustomRotator}.Rotated Values
   Tex=@n51
  ) {samplerType=SAMPLERTYPE_Masks, texture=&T_BlackMask}
  n53 = ChannelMaskParameter(None=@n52.RGBA) {parameterName=ColorCoreTwirl_Channel, sortPriority=6, defaultValue=<rgba 0 0 1 0>}
  n54 = Multiply(A=@n53.R, B=VectorParameter() {parameterName=ColorCoreTwirl, defaultValue=<rgba 1 0.6613 0.2402 1>}.RGB)
  n55 = ChannelMaskParameter(None=@n52.RGBA) {maskChannel=Blue, parameterName=ColorCoreTwirlb_Channel, sortPriority=6, defaultValue=<rgba 0 0 1 0>}
  n56 = Multiply(A=VectorParameter() {parameterName=ColorCoreTwirlb, defaultValue=<rgba 1.5 0.7112 0.0605 1>}.RGB, B=@n55.R)
  n57 = MaterialFunctionCall(
   UVs=@n36.UVs
   Rotation Center=@n48
   "Rotation Angle (0-1)"=Multiply(A=@n50, B=ScalarParameter() {defaultValue=0.02, parameterName=ColorCoreTwirl2_RotationSpeed})
  ) {materialFunction=&CustomRotator}
  n58 = TextureSample(UVs=@n57.Rotated Values, Tex=@n51) {samplerType=SAMPLERTYPE_Masks, texture=&T_BlackMask}
  n59 = ChannelMaskParameter(None=@n58.RGBA) {maskChannel=Green, parameterName=ColorCoreTwirl2_Channel, sortPriority=6, defaultValue=<rgba 0 1 0 0>}
  n60 = Multiply(A=@n59.R, B=VectorParameter() {parameterName=ColorCoreTwirl2, defaultValue=<rgba 0.8938 1 0 1>}.RGB)
  n61 = ChannelMaskParameter(None=@n58.RGBA) {maskChannel=Blue, parameterName=ColorCoreTwirl2b_Channel, sortPriority=6, defaultValue=<rgba 0 0 1 0>}
  n62 = Multiply(A=@n61.R, B=VectorParameter() {parameterName=ColorCoreTwirl2b, defaultValue=<rgba 0.2552 0.0433 0.0229 1>}.RGB)
  n63 = Power(Base=Add(A=Add(A=@n54, B=@n56), B=Add(A=@n60, B=@n62)), Exp=ScalarParameter() {defaultValue=7.02, parameterName=ContrastZeroTwirl})
  n64 = ScalarParameter() {defaultValue=-1.1005, parameterName=BrightnesSaturation}
  n65 = Add(
   A=Multiply(A=OneMinus(None=@n9), B=@n47)
   B=Multiply(A=Desaturation(None=@n63, Fraction=@n64), B=ScalarParameter() {defaultValue=0.3333, parameterName=Brightnesswirl})
  )
  n66 = ScalarParameter() {defaultValue=0.02, parameterName=ColorCoreTwirl3_RotationSpeed}
  n67 = TextureSample(UVs=MaterialFunctionCall(UVs=@n36.UVs, "Rotation Angle (0-1)"=Multiply(A=@n50, B=@n66)) {materialFunction=&CustomRotator}.Rotated Values) {texture=/Game/Characters/Player/Female/Medium/Bodies/F_MED_Celestial_Glow/Textures/T_FN_Nebula_Blue}
  n68 = ScalarParameter() {defaultValue=1, parameterName=nebulaBrightness}
  Add(A=@n65, B=Multiply(A=Multiply(A=@n67.RGB, B=@n68), B=@n46))
  Add(A=@n22) {constB=0}
  n69 = Multiply(A=@n23, B=ScalarParameter() {defaultValue=1, group=ZeroPoint_ExteriorNoise, parameterName=ZP_ExteriorStars1_Tiling})
  n70 = ScalarParameter() {defaultValue=-0.1, group=ZeroPoint_ExteriorNoise, parameterName=ZP_ExteriorStars1_panSpeed}
  n71 = ChannelMaskParameter(None=TextureSample(UVs=Panner(Coordinate=@n69, Speed=@n70), Tex=@n26) {texture=&T_BlacksRGB}.RGBA) {group=ZeroPoint_ExteriorNoise, parameterName=ZP_ExteriorStars1_Channel, sortPriority=6}
  n72 = Multiply(A=@n23, B=ScalarParameter() {defaultValue=1, group=ZeroPoint_ExteriorNoise, parameterName=ZP_ExteriorStars2_ Tiling})
  n73 = ScalarParameter() {defaultValue=0.2, group=ZeroPoint_ExteriorNoise, parameterName=ZP_ExteriorStars2_panSpeed}
  n74 = ChannelMaskParameter(None=TextureSample(UVs=Panner(Coordinate=@n72, Speed=@n73), Tex=@n26) {texture=&T_BlacksRGB}.RGBA) {group=ZeroPoint_ExteriorNoise, parameterName=ZP_ExteriorStars2_Channel, sortPriority=6}
  n75 = TextureCoordinate() {uTiling=2, vTiling=2}
  n76 = ChannelMaskParameter(None=TextureSample(UVs=Panner(Coordinate=@n75) {speedX=0.015, speedY=-0.04}, Tex=@n26) {texture=&T_BlacksRGB}.RGBA) {group=ZeroPoint_ExteriorNoise, maskChannel=Green, parameterName=ZP_StarNoise_ChannelMask, sortPriority=99, defaultValue=<rgba 0 1 0 0>}
  n77 = Multiply(
   A=Multiply(A=Add(A=@n71.R, B=@n74.R), B=@n76.R)
   B=VectorParameter() {group=ZeroPoint_ExteriorNoise, parameterName=ZP_StarColor, defaultValue=<rgba 1.852 0 20 1>}.RGB
  )
  n78 = Add(A=Add(A=@n53.R, B=@n55.R), B=Add(A=@n59.R, B=@n61.R))
  n79 = ScalarParameter() {defaultValue=8, parameterName=Fresnel -----}
  n80 = FunctionInput() {inputName=Material, inputType=FunctionInput_MaterialAttributes, sortPriority=-200}
  n81 = Transform(None=GetMaterialAttributes(None=@n80) {attributeGetTypes [#1, #2, #3]}.Normal)
  n82 = Multiply(
   A=Multiply(
    A=VectorParameter() {parameterName=Fresnel_Color, defaultValue=<rgba 1 1 1 1>}.RGB
    B=Fresnel(ExponentIn=@n79, Normal=@n81) {baseReflectFraction=0}
   )
   B=ScalarParameter() {defaultValue=1, parameterName=Fresnel Intensity}
  )
  n83 = VectorParameter() {parameterName=HoleRimColor, defaultValue=<rgba 0 0.1337 4 1>}
  n84 = SphereMask(
   A=Add(A=Multiply(A=@n31) {constB=0.03}, B=@n37)
   B=Constant()
   Radius=ScalarParameter() {defaultValue=0.175, parameterName=RimRadius}
   Hardness=ScalarParameter() {defaultValue=0.5, parameterName=Rim}
  )
  n85 = Multiply(A=@n84) {constB=4}
  n86 = Constant()
  n87 = ScalarParameter() {defaultValue=0.165, parameterName=CenterSize}
  n88 = Constant() {r=1}
  n89 = LinearInterpolate(
   A=LinearInterpolate(A=@n65, B=@n83.RGB, Alpha=Clamp(None=@n85))
   B=Multiply(A=@n47, B=ScalarParameter() {defaultValue=0.02, parameterName=BlackHoleTint})
   Alpha=Saturate(None=SphereMask(A=@n37, B=@n86, Radius=@n87, Hardness=@n88))
  ) {constB=0}
  n90 = MakeMaterialAttributes!(
   BaseColor=VectorParameter() {parameterName=BaseColor}.RGB
   Metallic=Constant()
   Specular=ScalarParameter() {defaultValue=0.5, parameterName=ZP2_Specular}
   Roughness=ScalarParameter() {defaultValue=0.25, parameterName=ZP2_Roughness}
   Anisotropy
   EmissiveColor=Add(A=LinearInterpolate(A=@n77, Alpha=@n78) {constB=0}, B=Add(A=@n82, B=@n89) {constB=0})
   Opacity
   OpacityMask
   Normal
   Tangent
   WorldPositionOffset
   SubsurfaceColor
   ClearCoat
   ClearCoatRoughness
   AmbientOcclusion
   Refraction
   CustomizedUVs_0
   CustomizedUVs_1
   CustomizedUVs_2
   CustomizedUVs_3
   CustomizedUVs_4
   CustomizedUVs_5
   CustomizedUVs_6
   CustomizedUVs_7
   PixelDepthOffset
   ShadingModel
   Displacement
  )
  n91 = FunctionInput() {inputName="Customization Mask (_CM B)", inputType=FunctionInput_Scalar, sortPriority=24, previewValue=<1 0 0 1>}
  n92 = FunctionInput() {inputName=Customization Mask B, inputType=FunctionInput_Scalar, sortPriority=24, previewValue=<1 0 0 1>}
  n93 = LinearInterpolate(A=LinearInterpolate(A=@n91, B=@n92, Alpha=NamedRerouteUsage()), B=Constant() {r=1}, Alpha=ScalarParameter() {parameterName=RollOverride})
  FunctionOutput(None=StaticSwitchParameter(True=BlendMaterialAttributes(A=@n80, B=@n90, Alpha=@n93), False=@n80) {group=Customization_Wrap, parameterName=CelestialGlow})
  n94 = VectorParameter() {parameterName=Axis}
  MaterialFunctionCall(
   Plane Normal Axis=@n7
   Plane X Axis=Normalize(VectorInput=CrossProduct(A=@n7, B=@n94.RGB))
   Plane Center=@n6
   UV Scale=Multiply(A=@n21, B=ScalarParameter() {defaultValue=0.2, parameterName=SizeProjectionTexture2Core})
  ) {materialFunction=&VirtualPlaneCoordinatesAbs}
  n95 = TransformPosition(None=Constant3Vector()) {transformType=TRANSFORMPOSSOURCE_World}
  n96 = Normalize(VectorInput=Subtract(A=@n95, B=CameraPositionWS()))
  n97 = Constant3Vector() {constant=<rgba 0 0 1 1>}
  n98 = FunctionInput() {bUsePreviewValueAsDefault=true, inputName=Object-SafeZone-Virtual-Pivot, inputType=FunctionInput_Vector4, sortPriority=105}
  n99 = Constant()
  n100 = LinearInterpolate(A=@n98, B=Add(A=@n98, B=AppendVector(B=@n99)), Alpha=Subtract(A=NamedRerouteUsage(), B=NamedRerouteUsage()))
  n101 = MaterialFunctionCall() {materialFunction=&ObjectScale}
  n102 = MaterialFunctionCall(Float3=PreSkinnedLocalBounds().Extents) {materialFunction=&BreakOutFloat3Components}
  n103 = Max(A=Max(A=@n102.R, B=@n102.R), B=@n102.B)
  MaterialFunctionCall(
   Plane Normal Axis=@n96
   Plane X Axis=Normalize(VectorInput=CrossProduct(A=@n96, B=@n97))
   Plane Center=Add(A=@n95, B=ComponentMask(None=@n100) {b=true, g=true, r=true})
   UV Scale=Multiply(A=ComponentMask(None=@n100) {a=true}, B=Multiply(A=@n101.Scale X, B=@n103))
  ) {materialFunction=&VirtualPlaneCoordinatesAbs}
  Multiply(A=@n83.RGB, B=@n84)
  Multiply(
   A=TextureSample(UVs=@n57.Rotated Values) {texture=&T_FN_Nebula_Blue}.RGB
   B=ScalarParameter() {defaultValue=0.4, parameterName=GalaxyBrightness}
  )
  n104 = MaterialFunctionCall(UV Channel=@n36.UVs) {materialFunction=&LinearGradient}
  Multiply(A=@n104.VGradient)
  n105 = FunctionInput() {bUsePreviewValueAsDefault=true, inputName=Object-SafeZone-Virtual-Pivot_CH_Switch, inputType=FunctionInput_Scalar, sortPriority=105}
  NamedRerouteDeclaration(None=MaterialFunctionCall!(
    Camera _pos
    Axis=VectorParameter() {parameterName=Axis_Char, defaultValue=<rgba 0 0 1 1>}.RGB
    Scale
    Scale_Multplyer
    Object-SafeZone-Virtual-Pivot=@n98
    Object-SafeZone-Virtual-CH-Switch=@n105
    SinglePlane Normal Axis
    SinglePlane X Axis
    PlaneCenter_distortion
   ) {materialFunction=&/MF_SafeZone_VirtualPlane}.Uvs) {name=SafeZone_UVs, nodeColor=<rgba 0.4297 1 0 1>}
  NamedRerouteDeclaration(None=FunctionInput() {bUsePreviewValueAsDefault=true, inputName=isCharacter, inputType=FunctionInput_Scalar}) {name=IsCharacter, nodeColor=<rgba 0.7944 1 0.0641 1>}
  n106 = FunctionInput() {bUsePreviewValueAsDefault=true, inputName=IsStatsB, inputType=FunctionInput_Vector2, sortPriority=-19}
  NamedRerouteDeclaration(None=ComponentMask(None=@n106) {r=true}) {name=IsBackpack, nodeColor=<rgba 0 0.9253 1 1>}
  NamedRerouteDeclaration(None=ComponentMask(None=@n106) {g=true}) {name=IsPickaxe, nodeColor=<rgba 0 0.9253 1 1>}
  NamedRerouteUsage()
  ObjectRadius()
  OneMinus(None=@n104.VGradient)
  n107 = Multiply(A=ScalarParameter() {defaultValue=10, parameterName=RotationSpeedCore}, B=Time())
  Rotator(Coordinate=@n36.UVs, Time=@n107) {speed=-0.25}
  Rotator(Coordinate=@n36.UVs, Time=@n107) {speed=0.2133}
  ScalarParameter() {defaultValue=1185.5463, parameterName=SizeProjectionTextureCore}
 ]}
