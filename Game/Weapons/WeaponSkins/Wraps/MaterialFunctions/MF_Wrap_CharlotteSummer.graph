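# Material function for the CharlotteSummer weapon wrap, layering an animated snake pattern over a customizable item's material behind a static switch.
# Randomly flipped Truchet tiles form arc UVs that scroll snake textures over projected backgrounds, with UV settings selected by item-type state inputs.
# The snake layer adds noise-driven glow and curve-atlas Fresnel metal reflectance; the accent layer adds thin-film iridescence.
# Primary, secondary and accent masks blend these layers.
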MaterialFunction /Game/Weapons/WeaponSkins/Wraps/MaterialFunctions/MF_Wrap_CharlotteSummer
graph {nodes [
  n1 = VectorParameter() {group=Wrap_CharlotteSummer, parameterName=CharS_Snake_C, defaultValue=<rgba 0.0273 0.5583 0.4564 1>}
  Add(A=Multiply(A=NamedRerouteUsage(), B=ScalarParameter() {group=Wrap_CharlotteSummer, parameterName=CharS_Snake_Iradecent_Strength}), B=@n1.RGB)
  n2 = NamedRerouteUsage()
  n3 = MaterialFunctionCall!(
   UVs=NamedRerouteUsage()
   Texture Size=ScalarParameter() {defaultValue=128, group=Wrap_CharlotteSummer_Snake, parameterName=CharS_Snake_Mip_Bias}
  ) {materialFunction=/Engine/Functions/Engine_MaterialFunctions01/Texturing/ComputeMipLevel}
  n4 = NamedRerouteUsage()
  n5 = Sign(None=Add(A=ComponentMask(None=@n4) {r=true}, B=ComponentMask(None=@n4) {g=true}))
  n6 = Multiply(A=@n5) {constB=0.5}
  n7 = Subtract(A=@n4, B=@n6)
  n8 = ComponentMask(None=@n7) {r=true}
  n9 = ComponentMask(None=@n7) {g=true}
  n10 = Constant() {r=0.25}
  n11 = Frac(None=Divide(
    A=Arctangent2Fast(Y=@n8, X=@n9)
    B=MaterialFunctionCall(Multiplier=@n10) {desc=Radial Gradient, materialFunction=&Pi}.Result
   ) {constB=1.571})
  n12 = ScalarParameter() {defaultValue=0.16, group=Wrap_CharlotteSummer_Snake, parameterName=CharS_Snake_Thickness}
  n13 = Subtract(A=@n4, B=@n6)
  n14 = Subtract(A=MaterialFunctionCall(Vector 2=Add(A=@n13) {constB=0}) {materialFunction=/Engine/Functions/Engine_MaterialFunctions02/Utility/VectorLength}.V2 Length) {constB=0.5}
  n15 = Divide(A=Subtract(A=@n12, B=@n14), B=Multiply(A=@n12) {constB=2})
  n16 = LinearInterpolate(A=@n15, B=OneMinus(None=@n15), Alpha=NamedRerouteUsage())
  n17 = LinearInterpolate(A=OneMinus(None=@n16), B=@n16, Alpha=NamedRerouteUsage())
  n18 = AppendVector(A=LinearInterpolate(A=OneMinus(None=@n11), B=@n11, Alpha=NamedRerouteUsage()), B=@n17)
  n19 = ScalarParameter() {defaultValue=-0.2, group=Wrap_CharlotteSummer_Snake, parameterName=CharS_Snake_Speed}
  n20 = Constant()
  n21 = Add(A=@n18, B=Multiply(A=Time(), B=AppendVector(A=@n19, B=@n20)))
  n22 = MaterialFunctionCall(Float4=NamedRerouteUsage()) {materialFunction=&BreakOutFloat4Components}
  n23 = LinearInterpolate(
   A=ScalarParameter() {defaultValue=5, group=Wrap_CharlotteSummer_Snake, parameterName=CharS_Snake_Normal_Strength}
   B=ScalarParameter() {defaultValue=5, group=Wrap_CharlotteSummer_Snake, parameterName=CharS_Snake_Normal_Strength_Vehicles}
   Alpha=@n22.R
  )
  n24 = ScalarParameter() {defaultValue=5, group=Wrap_CharlotteSummer_Snake, parameterName=CharS_Snake_Normal_Strength_Character}
  n25 = MaterialFunctionCall!(Value=Multiply(
    A=TextureSample!(UVs=@n21, Tex=@n2, Bias=@n3.MipLevel, Apply View MipBias) {mipValueMode=TMVM_MipBias, samplerType=SAMPLERTYPE_Masks}.R
    B=LinearInterpolate(A=@n23, B=@n24, Alpha=@n22.A)
   ) {constB=5}) {materialFunction=/Engine/Functions/Engine_MaterialFunctions03/Procedurals/PreparePerturbNormalHQ}
  AppendVector(A=AppendVector(A=Subtract(A=@n25.Value, B=@n25.Value one pixel right), B=Subtract(A=@n25.Value, B=@n25.Value one pixel down)), B=Constant())
  Arctangent2!(Y=@n8, X=@n9)
  ComponentMask(None=TextureCoordinate()) {r=true}
  Constant2Vector() {g=10.354, r=0.34}
  Constant2Vector() {g=435.354, r=234.34}
  Constant() {r=8}
  Constant() {r=0.34}
  n26 = Power(
   Base=DotProduct(A=CameraVectorWS(), B=PixelNormalWS())
   Exp=ScalarParameter() {defaultValue=0.5, group=Wrap_CharlotteSummer_Snake_Uv, parameterName=CharS_Snake_Iiridescent_Pow}
  )
  n27 = ScalarParameter() {defaultValue=2, group=Wrap_CharlotteSummer_Snake_Uv, parameterName=CharS_Snake_Iiridescent_TileRate}
  CurveAtlasRowParameter(CurveTime=Frac(None=Multiply(A=@n26, B=@n27))) {
   atlas=/Game/Weapons/WeaponSkins/Wraps/Textures/CruveAtlas_Wraps_01
   curve=/Game/Weapons/WeaponSkins/Wraps/Textures/0_ColorCurves/CC_Green_iridescent
   defaultValue=90
   group=Wrap_CharlotteSummer
   parameterName=CharS_Snake_Iiridescent_CC
  }
  n28 = TextureSample!(UVs, Tex, Apply View MipBias) {samplerType=SAMPLERTYPE_Masks, texture=/Game/Weapons/FORT_Rifles/Textures/T_Scar_CM}
  n29 = FunctionInput(Preview=@n28.R) {bUsePreviewValueAsDefault=true, inputName=Mask Primary, inputType=FunctionInput_Scalar, sortPriority=24, previewValue=<1 0 0 1>}
  n30 = FunctionInput(Preview=@n28.B) {bUsePreviewValueAsDefault=true, inputName=Mask Secondary, inputType=FunctionInput_Scalar, sortPriority=24, previewValue=<1 0 0 1>}
  n31 = BlendMaterialAttributes(A=NamedRerouteUsage(), B=NamedRerouteUsage(), Alpha=Saturate(None=Add(A=@n29, B=@n30))) {vertexAttributeBlendType=UseA}
  n32 = NamedRerouteUsage()
  n33 = NamedRerouteDeclaration(None=@n30) {name=Msk_Accent, nodeColor=<rgba 0 1 0.7656 1>}
  n34 = TextureSample() {samplerType=SAMPLERTYPE_Normal, texture=/Game/Weapons/FORT_Rifles/T_SCAR_N}
  n35 = Constant3Vector() {constant=<rgba 1 0 0 0>}
  n36 = GetMaterialAttributes(None=FunctionInput(Preview=SetMaterialAttributes!(MaterialAttributes, Normal=@n34.RGB, BaseColor=@n35) {attributeSetTypes [#1, #2]}) {inputName=Material, inputType=FunctionInput_MaterialAttributes}) {attributeGetTypes [
    #2
    #3
    #1
    #4
   ]}
  FunctionOutput(None=StaticSwitchParameter(
    True=BlendMaterialAttributes(A=@n31, B=@n32, Alpha=@n33) {vertexAttributeBlendType=UseA}
    False=NamedRerouteDeclaration(None=@n36.MaterialAttributes) {name=Root_Material, nodeColor=<rgba 0.7812 1 0 1>}
   ) {group=Customization_Wrap, parameterName=Wrap_CharlotteSummer})
  LinearInterpolate(
   A=Constant3Vector() {constant=<rgba 1 0 0 0>}
   B=Constant3Vector() {constant=<rgba 0 0 1 0>}
   Alpha=MaterialFunctionCall!(
    Blend Range=ScalarParameter() {defaultValue=600, group=Wrap_CharlotteSummer_Snake, parameterName=CharS_Snake_Distance}
    Start Offset=ScalarParameter() {defaultValue=-100, group=Wrap_CharlotteSummer_Snake, parameterName=CharS_Snake_StartOffset}
   ) {materialFunction=/Engine/Functions/Engine_MaterialFunctions03/Particles/Distance_Blend}.Result
  )
  n37 = NamedRerouteUsage()
  n38 = ScalarParameter() {defaultValue=1, parameterName=SmokeBlendAmt}
  LinearInterpolate(Alpha=Clamp(None=Multiply(A=@n37, B=@n38)))
  n39 = NamedRerouteUsage()
  MaterialFunctionCall(Input=@n39) {materialFunction=&RemapValueRange}
  Multiply(A=@n17)
  n40 = TextureSample!(UVs=@n21, Tex=NamedRerouteUsage(), Bias=@n3.MipLevel, Apply View MipBias) {mipValueMode=TMVM_MipBias}
  n41 = LinearInterpolate(
   A=@n1.RGB
   B=VectorParameter() {group=Wrap_CharlotteSummer, parameterName=CharS_Snake_A, defaultValue=<rgba 0.05 0.05 0.05 1>}.RGB
   Alpha=@n40.R
  )
  n42 = Multiply(A=NamedRerouteUsage(), B=ScalarParameter() {defaultValue=1, group=Wrap_CharlotteSummer_Glow, parameterName=CharS_Glow_SmokeBlendAmt})
  Multiply(A=LinearInterpolate(A=@n41, B=NamedRerouteUsage(), Alpha=Saturate(None=@n42)))
  Multiply(A=MaterialFunctionCall(Value_Morning=Constant() {r=0.5}, Value_Day=Constant() {r=1}, Value_Evening=Constant() {r=0.5}, Value_Night=Constant() {r=0.25}) {materialFunction=&Scalar_Time_of_Day}.Result)
  Multiply() {constA=1.571, constB=2}
  Multiply(A=@n3.MipLevel)
  n43 = VectorParameter() {group=Wrap_CharlotteSummer_Accent, parameterName=CharS_Accent_BaseColor, defaultValue=<rgba 0.003 0.003 0.003 1>}
  n44 = SetMaterialAttributes!(
   MaterialAttributes=NamedRerouteUsage()
   BaseColor=@n43.RGB
   Roughness=ScalarParameter() {defaultValue=0.3, group=Wrap_CharlotteSummer_Accent, parameterName=CharS_Accent_Roughness}
   Specular=ScalarParameter() {defaultValue=0.5, group=Wrap_CharlotteSummer_Accent, parameterName=CharS_Accent_Spec}
   Metallic=ScalarParameter() {defaultValue=0.5, group=Wrap_CharlotteSummer_Accent, parameterName=CharS_Accent_Metalic}
   Normal
   EmissiveColor
  ) {attributeSetTypes [
    #2
    #3
    #5
    #4
    #1
    #6
   ]}
  n45 = GetMaterialAttributes(None=@n44) {attributeGetTypes [#3, #2]}
  n46 = ScalarParameter() {defaultValue=0.5, group=Wrap_CharlotteSummer_Accent, parameterName=CharS_Accent_ThinFilm_Warp}
  n47 = TextureSampleParameter2D(UVs=Multiply(
    A=ScalarParameter() {defaultValue=1, group=Wrap_CharlotteSummer_Accent, parameterName=CharS_Accent_ThinFilm_Scale}
    B=Fresnel(ExponentIn=@n46) {baseReflectFraction=0}
   )) {
   group=Wrap_CharlotteSummer_Accent
   parameterName=CharS_Accent_Thin Film Texture
   texture=/Game/Global/Textures/Default/Thinfilm/T_ThinFilm_Spectrum_COLOR
  }
  n48 = Add(
   A=@n45.BaseColor
   B=Multiply(A=@n47.RGB, B=ScalarParameter() {defaultValue=0.5, group=Wrap_CharlotteSummer_Accent, parameterName=CharS_Accent_ThinFilm_Intensity})
  )
  n49 = Saturate(None=Multiply(
    A=OneMinus(None=@n45.Roughness)
    B=ScalarParameter() {defaultValue=1, group=Wrap_CharlotteSummer_Accent, parameterName=CharS_Accent_ThinFilm_RoughnessScale}
   ))
  n50 = Fresnel(ExponentIn=ScalarParameter() {defaultValue=1.2164, group=Wrap_CharlotteSummer_Accent, parameterName=CharS_Accent_ThinFilm_Exponent}) {baseReflectFraction=0}
  n51 = Saturate(None=Multiply(A=Multiply(A=@n49, B=@n50), B=@n39))
  NamedRerouteDeclaration(None=SetMaterialAttributes!(MaterialAttributes=@n45.MaterialAttributes, BaseColor=LinearInterpolate(A=@n45.BaseColor, B=@n48, Alpha=@n51)) {attributeSetTypes [#2]}) {name=Mat_Accent, nodeColor=<rgba 0 0.3906 1 1>}
  n52 = FunctionInput(Preview=@n28.G) {bUsePreviewValueAsDefault=true, inputName=Mask Crevise, inputType=FunctionInput_Scalar, sortPriority=24, previewValue=<1 0 0 1>}
  NamedRerouteDeclaration(None=@n52) {name=Msk_Curvature, nodeColor=<rgba 0 1 0.7422 1>}
  NamedRerouteDeclaration(None=TextureObjectParameter() {group=Wrap_CharlotteSummer, parameterName=CharS__Tex_Scale_Msk, texture=&T_BlacksRGB}) {name=Tex_Scale_Msk, nodeColor=<rgba 0.3125 1 0 1>}
  n53 = MaterialFunctionCall(Float4=NamedRerouteUsage()) {materialFunction=&BreakOutFloat4Components}
  n54 = MaterialFunctionCall!(
   For_Safe_Zones
   Normal_Map
   Caculated_Bounds=NamedRerouteUsage()
   Pre_Skinned_Normals=NamedRerouteUsage()
   Is Vehicle=@n53.R
   Is VehicleSide=@n53.R
   Is Mech=@n53.R
   Vehicle_Rotation
   UV_Rotation Bypass
   Mech_Rotation
   Uvs_Offset
   Offset-after rotation
   Uvs_TileRate
  ) {materialFunction=&/MF_Projection_Side_Normal}
  n55 = Multiply(A=Add(A=NamedRerouteDeclaration(None=@n54.UVs) {name=UVs, nodeColor=<rgba 0.2656 0 1 1>}, B=NamedRerouteUsage()), B=NamedRerouteUsage()) {constB=5}
  n56 = Floor(None=@n55)
  n57 = Multiply(A=Add(A=@n56) {constB=-0.5}, B=Add(A=Constant2Vector() {g=3.354, r=0.34}, B=NamedRerouteUsage()))
  n58 = DotProduct(A=Frac(None=@n57), B=Add(A=@n57) {constB=34.23})
  n59 = Frac(None=Multiply(A=ComponentMask(None=@n58) {r=true}, B=ComponentMask(None=@n58) {g=true}))
  NamedRerouteDeclaration(None=If(A=@n59, B=Constant() {r=0.5}, "A > B"=Constant(), "A < B"=Constant() {r=1})) {name=Truch_Rand_Mask, nodeColor=<rgba 0.125 0 1 1>}
  n60 = TextureObjectParameter() {
   group=Wrap_CharlotteSummer
   parameterName=CharS__Tex_BkG_Nrml
   samplerType=SAMPLERTYPE_Normal
   texture=&FlatNormalBlue
  }
  NamedRerouteDeclaration(None=@n60) {name=Tex_Scale_Nrml, nodeColor=<rgba 0.3125 1 0 1>}
  NamedRerouteDeclaration(None=SmoothStep(Min=Add(A=@n12) {constB=0.001}, Max=@n12, Value=Abs(None=@n14)) {constMax=0.01, constMin=0.02}) {name=Truch_Masking, nodeColor=<rgba 0 1 0.9766 1>}
  NamedRerouteDeclaration(None=@n56) {name=UV_ID, nodeColor=<rgba 1 0 0.9141 1>}
  NamedRerouteDeclaration(None=Saturate(None=@n5)) {name=Msk_Troch, nodeColor=<rgba 0 0.7656 1 1>}
  n61 = NamedRerouteUsage()
  n62 = NamedRerouteUsage()
  n63 = VectorParameter() {group=Wrap_CharlotteSummer, parameterName=CharS_Snake_B, defaultValue=<rgba 0.05 0.05 0.05 1>}
  n64 = NamedRerouteUsage()
  n65 = NamedRerouteUsage()
  n66 = Multiply(A=@n65, B=@n40.B) {constA=1}
  n67 = LinearInterpolate(
   A=ScalarParameter() {defaultValue=0.1, group=Wrap_CharlotteSummer, parameterName=CharS_Snake_Roughness_C}
   B=Add(A=ScalarParameter() {defaultValue=0.25, group=Wrap_CharlotteSummer, parameterName=CharS_Snake_Roughness_A}, B=@n40.A)
   Alpha=@n40.R
  )
  n68 = ScalarParameter() {defaultValue=0.9, group=Wrap_CharlotteSummer, parameterName=CharS_Snake_Roughness_B}
  n69 = Multiply(A=@n66, B=NamedRerouteUsage()) {constA=1}
  n70 = ScalarParameter() {group=Wrap_CharlotteSummer, parameterName=CharS_Snake_MetalicC}
  n71 = OneMinus(None=@n40.G)
  n72 = ScalarParameter() {defaultValue=5, group=Wrap_CharlotteSummer, parameterName=CharS_Snake_Metalic_power}
  n73 = LinearInterpolate(
   A=ScalarParameter() {defaultValue=1, group=Wrap_CharlotteSummer, parameterName=CharS_Snake_MetalicA}
   B=ScalarParameter() {defaultValue=0.5, group=Wrap_CharlotteSummer, parameterName=CharS_Snake_MetalicB}
   Alpha=Saturate(None=Power(Base=@n71, Exp=@n72) {constExponent=10})
  )
  n74 = Multiply(A=@n40.R, B=NamedRerouteUsage())
  n75 = NamedRerouteDeclaration(None=Saturate(None=Multiply(A=@n65, B=@n74))) {name=Msk texture, nodeColor=<rgba 1 0.9609 0 1>}
  n76 = Multiply(
   A=Transform(None=MaterialFunctionCall!(
     Bump one pixel right=@n25.Value one pixel right
     Bump center=@n25.Value one pixel right
     Bump one pixel down=@n25.Value one pixel right
     WS Normal
    ) {materialFunction=/Engine/Functions/Engine_MaterialFunctions03/Procedurals/PerturbNormalHQ}.WS Normal) {transformSourceType=TRANSFORMSOURCE_World, transformType=TRANSFORM_Tangent}
   B=Constant3Vector() {constant=<rgba 1 1 0 0>}
  )
  n77 = Multiply(A=@n66, B=NamedRerouteUsage()) {constA=1}
  n78 = NamedRerouteUsage()
  n79 = NamedRerouteUsage()
  n80 = SetMaterialAttributes!(
   MaterialAttributes=@n61
   BaseColor=LinearInterpolate(
    A=ComponentMask(None=@n62) {b=true, g=true, r=true}
    B=LinearInterpolate(A=@n41, B=@n63.RGB, Alpha=@n40.G)
    Alpha=Multiply(A=@n64, B=@n66) {constA=1}
   )
   Roughness=LinearInterpolate(A=ComponentMask(None=@n62) {a=true}, B=LinearInterpolate(A=@n67, B=@n68, Alpha=@n40.G), Alpha=@n69)
   Specular=LinearInterpolate(
    A=ScalarParameter() {defaultValue=0.5, group=Wrap_CharlotteSummer, parameterName=CharS_Spec_BkG}
    B=ScalarParameter() {defaultValue=0.5, group=Wrap_CharlotteSummer, parameterName=CharS_Spec_Snake}
    Alpha=@n69
   )
   Metallic=LinearInterpolate(
    A=ScalarParameter() {group=Wrap_CharlotteSummer, parameterName=CharS_Snake_Metalic_Bck}
    B=LinearInterpolate(A=@n70, B=@n73, Alpha=@n75)
    Alpha=@n69
   )
   Normal=Add(A=NamedRerouteUsage(), B=LinearInterpolate(B=@n76, Alpha=@n77))
   EmissiveColor=Multiply(A=Multiply(A=@n78, B=@n79), B=@n40.B)
  ) {attributeSetTypes [
    #2
    #3
    #5
    #4
    #1
    #6
   ]}
  n81 = GetMaterialAttributes(None=@n80) {attributeGetTypes [#1, #2]}
  n82 = CameraVectorWS()
  n83 = PixelNormalWS()
  n84 = Saturate(None=Power(
    Base=DotProduct(A=@n82, B=@n83)
    Exp=ScalarParameter() {defaultValue=1, group=Wrap_CharlotteSummer_MetalLUT, parameterName=CharS_FresnelReflectance_Power}
   ))
  n85 = CurveAtlasRowParameter(CurveTime=@n84) {
   atlas=&CruveAtlas_Wraps_01
   curve=/Game/Weapons/WeaponSkins/Wraps/Textures/0_ColorCurves/AluminiumreflectanceCurve
   defaultValue=91
   group=Wrap_CharlotteSummer_MetalLUT
   parameterName=CharS_FresnelReflectance_LUT
  }
  n86 = ChannelMaskParameter(None=MaterialFunctionCall!(
    Index=ScalarParameter() {defaultValue=91, group=Wrap_CharlotteSummer_MetalLUT, parameterName="CharS_FresnelReflectance_LUT  Index"}
    CCinput=@n85
    CCinput_Alpha=@n85
    UV=@n84
   ) {materialFunction=&/MF_Global_CC_Tex}.4v) {group=Wrap_CharlotteSummer_MetalLUT, maskChannel=Green, parameterName=CharS_LUTChannel, defaultValue=<rgba 0 1 0 0>}
  n87 = ScalarParameter() {defaultValue=2.1, group=Wrap_CharlotteSummer_MetalLUT, parameterName=CharS_FresnelReflectance_Max}
  n88 = Multiply(
   A=@n81.BaseColor
   B=MaterialFunctionCall(
    Input=@n86.R
    Input Low=Constant()
    Input High=@n87
    Target Low=ScalarParameter() {defaultValue=0.258, group=Wrap_CharlotteSummer_MetalLUT, parameterName=CharS_FresnelReflectance_Min}
    Target High=@n87
   ) {materialFunction=&RemapValueRange}.Result
  )
  n89 = NamedRerouteUsage()
  NamedRerouteDeclaration(None=SetMaterialAttributes!(MaterialAttributes=@n81.MaterialAttributes, BaseColor=LinearInterpolate(A=@n81.BaseColor, B=@n88, Alpha=@n89)) {attributeSetTypes [#2]}) {name=Mat_Base, nodeColor=<rgba 0 0.3906 1 1>}
  NamedRerouteDeclaration(None=TextureObjectParameter() {group=Wrap_CharlotteSummer, parameterName=CharS__Tex_Scale_Msk2, texture=&T_BlacksRGB}) {name=Tex_Scale_Msk2, nodeColor=<rgba 0.3125 1 0 1>}
  n90 = Add(A=Frac(None=@n55)) {constB=-0.5, desc=gv}
  n91 = Constant() {r=0.5}
  n92 = ComponentMask(None=@n90) {g=true}
  n93 = Multiply(A=@n92) {constB=-1}
  NamedRerouteDeclaration(None=AppendVector(A=ComponentMask(None=@n90) {r=true}, B=If(A=@n59, B=@n91, "A > B"=@n92, "A < B"=@n93))) {name=GV, nodeColor=<rgba 1 0 0.9141 1>}
  n94 = ScalarParameter() {defaultValue=10, group=Wrap_CharlotteSummer_Snake_Uv, parameterName=CharS_Snake_TileRate_Vehicle, sortPriority=150}
  n95 = ScalarParameter() {defaultValue=10, group=Wrap_CharlotteSummer_Snake_Uv, parameterName=CharS_Snake_TileRate_Vehicle_Side, sortPriority=150}
  n96 = MaterialFunctionCall(Float4=NamedRerouteUsage()) {materialFunction=&BreakOutFloat4Components}
  n97 = LinearInterpolate(
   A=LinearInterpolate(
    A=ScalarParameter() {defaultValue=10, group=Wrap_CharlotteSummer_Snake_Uv, parameterName=CharS_Snake_TileRate_Weapon, sortPriority=150}
    B=LinearInterpolate(A=@n94, B=@n95, Alpha=@n96.G)
    Alpha=@n96.R
   )
   B=LinearInterpolate(
    A=ScalarParameter() {defaultValue=10, group=Wrap_CharlotteSummer_Snake_Uv, parameterName=CharS_Snake_TileRate_Mech, sortPriority=150}
    B=ScalarParameter() {defaultValue=10, group=Wrap_CharlotteSummer_Snake_Uv, parameterName=CharS_Snake_TileRate_Character, sortPriority=150}
    Alpha=@n96.A
   )
   Alpha=@n96.B
  )
  n98 = ScalarParameter() {defaultValue=10, group=Wrap_CharlotteSummer_Snake_Uv, parameterName=CharS_Snake_TileRate_Backpack, sortPriority=150}
  n99 = MaterialFunctionCall(Float2=NamedRerouteUsage()) {materialFunction=&BreakOutFloat2Components}
  NamedRerouteDeclaration(None=LinearInterpolate(
    A=LinearInterpolate(A=@n97, B=@n98, Alpha=@n99.R)
    B=ScalarParameter() {defaultValue=10, group=Wrap_CharlotteSummer_Snake_Uv, parameterName=CharS_Snake_TileRate_Pickaxe, sortPriority=150}
    Alpha=@n99.G
   )) {name=UVs_Snake_Scale, nodeColor=<rgba 0 0.7188 1 1>}
  n100 = VectorParameter() {group=Wrap_CharlotteSummer_Snake_Uv, parameterName=CharS_Snake_Offset_Weapon, channelNames {a=Seed, b="", g="", r=""}}
  n101 = VectorParameter() {group=Wrap_CharlotteSummer_Snake_Uv, parameterName=CharS_Snake_Offset_Vehicle, channelNames {a=Seed, b="", g="", r=""}}
  n102 = VectorParameter() {group=Wrap_CharlotteSummer_Snake_Uv, parameterName=CharS_Snake_Offset_VehicleSide, channelNames {a=Seed, b="", g="", r=""}}
  n103 = MaterialFunctionCall(Float4=NamedRerouteUsage()) {materialFunction=&BreakOutFloat4Components}
  n104 = LinearInterpolate(
   A=AppendVector(A=@n100, B=@n100)
   B=LinearInterpolate(A=AppendVector(A=@n101, B=@n101), B=AppendVector(A=@n102, B=@n102), Alpha=@n103.G)
   Alpha=@n103.R
  )
  n105 = VectorParameter() {group=Wrap_CharlotteSummer_Snake_Uv, parameterName=CharS_Snake_Offset_Mech, channelNames {a=Seed, b="", g="", r=""}}
  n106 = VectorParameter() {group=Wrap_CharlotteSummer_Snake_Uv, parameterName=CharS_Snake_Offset_Character, channelNames {a=Seed, b="", g="", r=""}}
  n107 = LinearInterpolate(A=AppendVector(A=@n105, B=@n105), B=AppendVector(A=@n106, B=@n106), Alpha=@n103.A)
  n108 = VectorParameter() {group=Wrap_CharlotteSummer_Snake_Uv, parameterName=CharS_Snake_Offset_Backpack, channelNames {a=Seed, b="", g="", r=""}}
  n109 = MaterialFunctionCall(Float2=NamedRerouteUsage()) {materialFunction=&BreakOutFloat2Components}
  n110 = VectorParameter() {group=Wrap_CharlotteSummer_Snake_Uv, parameterName=CharS_Snake_Offset_Pickaxe, channelNames {a=Seed, b="", g="", r=""}}
  n111 = LinearInterpolate(
   A=LinearInterpolate(A=LinearInterpolate(A=@n104, B=@n107, Alpha=@n103.B), B=AppendVector(A=@n108, B=@n108), Alpha=@n109.R)
   B=AppendVector(A=@n110, B=@n110)
   Alpha=@n109.G
  )
  NamedRerouteDeclaration(None=ComponentMask(None=@n111) {g=true, r=true}) {name=UVs_Snake_Offset, nodeColor=<rgba 0 0.625 1 1>}
  n112 = ScalarParameter() {defaultValue=1, group=Wrap_CharlotteSummer_UVs, parameterName=CharS_Scale_BkG_Vehicle, sortPriority=150}
  n113 = ScalarParameter() {defaultValue=1, group=Wrap_CharlotteSummer_UVs, parameterName=CharS_Scale_BkG_Vehicle_Side, sortPriority=150}
  n114 = MaterialFunctionCall(Float4=NamedRerouteUsage()) {materialFunction=&BreakOutFloat4Components}
  n115 = LinearInterpolate(
   A=LinearInterpolate(
    A=ScalarParameter() {defaultValue=1, group=Wrap_CharlotteSummer_UVs, parameterName=CharS_Scale_BkG_Weapon, sortPriority=150}
    B=LinearInterpolate(A=@n112, B=@n113, Alpha=@n114.G)
    Alpha=@n114.R
   )
   B=LinearInterpolate(
    A=ScalarParameter() {defaultValue=1, group=Wrap_CharlotteSummer_UVs, parameterName=CharS_Scale_BkG_Mech, sortPriority=150}
    B=ScalarParameter() {defaultValue=1, group=Wrap_CharlotteSummer_UVs, parameterName=CharS_Scale_BkG_Character, sortPriority=150}
    Alpha=@n114.A
   )
   Alpha=@n114.B
  )
  n116 = ScalarParameter() {defaultValue=1, group=Wrap_CharlotteSummer_UVs, parameterName=CharS_Scale_BkG_Backpack, sortPriority=150}
  n117 = MaterialFunctionCall(Float2=NamedRerouteUsage()) {materialFunction=&BreakOutFloat2Components}
  NamedRerouteDeclaration(None=LinearInterpolate(
    A=LinearInterpolate(A=@n115, B=@n116, Alpha=@n117.R)
    B=ScalarParameter() {defaultValue=1, group=Wrap_CharlotteSummer_UVs, parameterName=CharS_Scale_BkG_Pickaxe, sortPriority=150}
    Alpha=@n117.G
   )) {name=UVs_BkG_Scale, nodeColor=<rgba 0 0.7188 1 1>}
  n118 = MaterialFunctionCall(Color=Abs(None=NamedRerouteUsage())) {materialFunction=&SplitComponents}
  n119 = MaterialFunctionCall(Float4=NamedRerouteUsage()) {materialFunction=&BreakOutFloat4Components}
  n120 = LinearInterpolate(A=LinearInterpolate(A=@n118.R, B=LinearInterpolate(A=@n118.B, B=@n118.B, Alpha=@n119.G), Alpha=@n119.R), B=@n118.G, Alpha=@n119.B)
  n121 = ScalarParameter() {defaultValue=0.5, group=Wrap_CharlotteSummer, parameterName=CharS_Blend_Contrast}
  NamedRerouteDeclaration(None=LinearInterpolate(
    A=MaterialFunctionCall(In=@n120, Contrast=@n121) {materialFunction=&CheapContrast}.Result
    Alpha=@n119.A
   )) {name=Tri_Mask, nodeColor=<rgba 1 0.4688 0 1>}
  n122 = VectorParameter() {group=Wrap_CharlotteSummer_UVs, parameterName=CharS_BkG_Offset_Vehicle, defaultValue=<rgba 0 0 0 1>}
  n123 = VectorParameter() {group=Wrap_CharlotteSummer_UVs, parameterName=CharS_BkG_Offset_VehicleSide, defaultValue=<rgba 0 0 0 1>}
  n124 = MaterialFunctionCall(Float4=NamedRerouteUsage()) {materialFunction=&BreakOutFloat4Components}
  n125 = LinearInterpolate(
   A=LinearInterpolate(
    A=VectorParameter() {group=Wrap_CharlotteSummer_UVs, parameterName=CharS_BkG_Offset_Weapon, defaultValue=<rgba 0 0 0 1>}.RGB
    B=LinearInterpolate(A=@n122.RGB, B=@n123.RGB, Alpha=@n124.G)
    Alpha=@n124.R
   )
   B=LinearInterpolate(
    A=VectorParameter() {group=Wrap_CharlotteSummer_UVs, parameterName=CharS_BkG_Offset_Mech, defaultValue=<rgba 0 0 0 1>}.RGB
    B=VectorParameter() {group=Wrap_CharlotteSummer_UVs, parameterName=CharS_BkG_Offset_Character, defaultValue=<rgba 0 0 0 1>}.RGB
    Alpha=@n124.A
   )
   Alpha=@n124.B
  )
  n126 = VectorParameter() {group=Wrap_CharlotteSummer_UVs, parameterName=CharS_BkG_Offset_Backpack, defaultValue=<rgba 0 0 0 1>}
  n127 = MaterialFunctionCall(Float2=NamedRerouteUsage()) {materialFunction=&BreakOutFloat2Components}
  NamedRerouteDeclaration(None=LinearInterpolate(
    A=LinearInterpolate(A=@n125, B=@n126.RGB, Alpha=@n127.R)
    B=VectorParameter() {group=Wrap_CharlotteSummer_UVs, parameterName=CharS_BkG_Offset_Pickaxe, defaultValue=<rgba 0 0 0 1>}.RGB
    Alpha=@n127.G
   )) {name=UVs_BkG_Offset, nodeColor=<rgba 0 0.625 1 1>}
  n128 = VectorParameter() {group=Wrap_CharlotteSummer_UVs, parameterName=CharS_BkG_Rotation_Weapon, defaultValue=<rgba 0 0 -0.25 1>}
  n129 = VectorParameter() {group=Wrap_CharlotteSummer_UVs, parameterName=CharS_BkG_Rotation_Vehicle, defaultValue=<rgba 0 0 0 1>}
  n130 = VectorParameter() {group=Wrap_CharlotteSummer_UVs, parameterName=CharS_BkG_Rotation_Vehicle_Side, defaultValue=<rgba 0 0 0 1>}
  n131 = MaterialFunctionCall(Float4=NamedRerouteUsage()) {materialFunction=&BreakOutFloat4Components}
  n132 = VectorParameter() {group=Wrap_CharlotteSummer_UVs, parameterName=CharS_BkG_Rotation_Mech, defaultValue=<rgba 0.25 0.25 0 1>}
  n133 = VectorParameter() {group=Wrap_CharlotteSummer_UVs, parameterName=CharS_BkG_Rotation_Character, defaultValue=<rgba 0.25 0.25 0 1>}
  n134 = LinearInterpolate(
   A=LinearInterpolate(A=@n128.RGB, B=LinearInterpolate(A=@n129.RGB, B=@n130.RGB, Alpha=@n131.G), Alpha=@n131.R)
   B=LinearInterpolate(A=@n132.RGB, B=@n133.RGB, Alpha=@n131.A)
   Alpha=@n131.B
  )
  n135 = VectorParameter() {group=Wrap_CharlotteSummer_UVs, parameterName=CharS_BkG_Rotation_Backpack, defaultValue=<rgba 0 0 0 1>}
  n136 = MaterialFunctionCall(Float2=NamedRerouteUsage()) {materialFunction=&BreakOutFloat2Components}
  NamedRerouteDeclaration(None=LinearInterpolate(
    A=LinearInterpolate(A=@n134, B=@n135.RGB, Alpha=@n136.R)
    B=VectorParameter() {group=Wrap_CharlotteSummer_UVs, parameterName=CharS_BkG_Rotation_Pickaxe, defaultValue=<rgba 0 0 0 1>}.RGB
    Alpha=@n136.G
   )) {name=UVs_BkG_Rotation, nodeColor=<rgba 0 0.7188 1 1>}
  NamedRerouteDeclaration(None=@n121) {name=Contrast, nodeColor=<rgba 1 0 0.8437 1>}
  n137 = NamedRerouteUsage()
  n138 = Constant()
  NamedRerouteDeclaration(None=MaterialFunctionCall!(
    Texture_Normals=NamedRerouteUsage()
    Rotation_Axis=@n137
    Caculated_Bounds=NamedRerouteUsage()
    Pre_Skinned_Normals=NamedRerouteUsage()
    IsStatsB=NamedRerouteUsage()
    IsState=NamedRerouteUsage()
    Blend_Contrast=NamedRerouteUsage()
    Fill_alternate_Axis_Opacity=@n138
    Scale=NamedRerouteUsage()
    Offset_Pivot
    Offset=NamedRerouteUsage()
   ) {materialFunction=&/MF_TopsSideProjection_Normals_V2}.Add_Surface_Normal) {name=TopSide_Normals, nodeColor=<rgba 0 0.9297 1 1>}
  NamedRerouteDeclaration(None=ComponentMask(None=@n111) {a=true}) {name=UVs_Snake_Seed, nodeColor=<rgba 0 0.625 1 1>}
  NamedRerouteDeclaration(None=@n55) {name=UV_tiles, nodeColor=<rgba 1 0.1406 0 1>}
  NamedRerouteDeclaration(None=TextureObjectParameter() {group=Wrap_CharlotteSummer, parameterName=CharS__Tex_BkG, texture=&T_BlacksRGB}) {name=Tex_BkG, nodeColor=<rgba 0.3125 1 0 1>}
  NamedRerouteDeclaration(None=MaterialFunctionCall!(
    Texture=NamedRerouteUsage()
    Caculated_Bounds=NamedRerouteUsage()
    Pre_Skinned_Normals=NamedRerouteUsage()
    Blend_Contrast=NamedRerouteUsage()
    IsStatsB=NamedRerouteUsage()
    IsState=NamedRerouteUsage()
    Fill_alternate_Axis_Opacity=@n138
    "Fill_alternate_Axis_Color_A=Roughness"=Constant4Vector() {constant=<rgba 0 0 0 1>}.RGBA
    Offset=NamedRerouteUsage()
    Scale=NamedRerouteUsage()
    Offset_Pivot
    Rotation_Axis=@n137
   ) {materialFunction=&/MF_TopsSideProjection_Base_V2}.Combined) {name=TopSide BkG, nodeColor=<rgba 0 1 0.1797 1>}
  NamedRerouteDeclaration(None=@n47.RGB) {name=Iradecent, nodeColor=<rgba 1 0 0.3984 1>}
  n139 = SphereMask(
   A=ComponentMask(None=NamedRerouteUsage()) {g=true}
   B=Constant() {r=0.5}
   Radius=ScalarParameter() {defaultValue=0.34, group=Wrap_CharlotteSummer_Glow, parameterName=CharS_Glow_Radius}
   Hardness=ScalarParameter() {group=Wrap_CharlotteSummer_Glow, parameterName=CharS_Glow_Hardness}
  )
  NamedRerouteDeclaration(None=Multiply(A=NamedRerouteUsage(), B=OneMinus(None=@n139))) {name=Y_Gradient, nodeColor=<rgba 1 0.2812 0 1>}
  NamedRerouteDeclaration(None=@n18) {name=UVs_Truc, nodeColor=<rgba 0 1 0.7188 1>}
  n140 = NamedRerouteUsage()
  n141 = ComponentMask(None=VectorParameter() {group=Wrap_CharlotteSummer_Glow, parameterName=CharS_Glow_TileRate, defaultValue=<rgba 1 1 0 1>}.RGB) {g=true, r=true}
  n142 = Time()
  n143 = ComponentMask(None=VectorParameter() {group=Wrap_CharlotteSummer_Glow, parameterName=CharS_Glow_Speed, defaultValue=<rgba 0 0 0 1>}.RGB) {g=true, r=true}
  n144 = TextureSampleParameter2D(UVs=Add(A=Multiply(A=@n140, B=@n141), B=Multiply(A=@n142, B=@n143))) {
   group=Wrap_CharlotteSummer_Glow
   parameterName=CharS_Glow__Tex
   samplerType=SAMPLERTYPE_LinearGrayscale
   texture=/Game/Global/Textures/Default/Blanks/T_BlackLinearGrayscale
  }
  n145 = Subtract(
   A=NamedRerouteUsage()
   B=Multiply(A=@n144.RGB, B=ScalarParameter() {defaultValue=0.5, group=Wrap_CharlotteSummer_Glow, parameterName=CharS_Glow_NoiseStrength})
  )
  n146 = VectorParameter() {group=Wrap_CharlotteSummer_Glow, parameterName=CharS_Glow_EmissiveColorSnake, defaultValue=<rgba 0 1 0.515 1>}
  n147 = Multiply(A=NamedRerouteDeclaration(None=Saturate(None=@n145)) {name=Glow_Mist, nodeColor=<rgba 0.2656 0 1 1>}, B=@n146.RGB)
  NamedRerouteDeclaration(None=Multiply(A=@n147, B=ScalarParameter() {defaultValue=0.5, group=Wrap_CharlotteSummer_Glow, parameterName=CharS_Glow_DiffuseMult})) {name=Glow_Difuse, nodeColor=<rgba 0.2656 0 1 1>}
  n148 = ScalarParameter() {defaultValue=0.1, group=Wrap_CharlotteSummer_Glow, parameterName=CharS_Glow_EmissiveStrengthSnake}
  NamedRerouteDeclaration(None=Power(
    Base=Multiply(A=@n147, B=@n148)
    Exp=ScalarParameter() {defaultValue=1, group=Wrap_CharlotteSummer_Glow, parameterName=CharS_Glow_FresnelPowerSnake}
   )) {name=Glow_Emissive, nodeColor=<rgba 0.2656 0 1 1>}
  NamedRerouteDeclaration(None=FunctionInput(Preview=Constant4Vector().RGBA) {bUsePreviewValueAsDefault=true, inputName=IsState, inputType=FunctionInput_Vector4, sortPriority=4}) {name=IsState, nodeColor=<rgba 0.1497 1 0.1633 1>}
  NamedRerouteDeclaration(None=FunctionInput(Preview=Constant2Vector()) {bUsePreviewValueAsDefault=true, inputName=IsState_B, inputType=FunctionInput_Vector2, sortPriority=4}) {name=IsState_B, nodeColor=<rgba 0.1497 1 0.1633 1>}
  n149 = PreSkinnedLocalBounds()
  n150 = Subtract(A=VertexInterpolator(VS=PreSkinnedPosition()).PS, B=@n149.Min)
  n151 = MaterialFunctionCall(Color=Subtract(A=@n149.Max, B=@n149.Max)) {materialFunction=&SplitComponents}
  n152 = Max(A=Max(A=@n151.R, B=@n151.R), B=@n151.B)
  NamedRerouteDeclaration(None=FunctionInput(Preview=Divide(A=@n150, B=@n152)) {bUsePreviewValueAsDefault=true, inputName=Calcated_Bounds, sortPriority=11}) {name=Caclcated_Bounds, nodeColor=<rgba 0 0.9297 1 1>}
  MaterialExpressionAdd_2 = Add(A=ComponentMask(None=@n56) {r=true}, B=ComponentMask(None=@n56) {g=true})
  MaterialExpressionConstant_3 = Constant() {r=2}
  n153 = Custom!(x=@MaterialExpressionAdd_2, y=@MaterialExpressionConstant_3) {
   code="   return x-y*floor(x/y);"
   description=Mod
   outputType=CMOT_Float1
   inputs [
    {inputName=x, input {expression=&self:MaterialExpressionAdd_2}}
    {inputName=y, input {expression=&self:MaterialExpressionConstant_3}}
   ]
  }
  NamedRerouteDeclaration(None=@n153) {name=Truch_checkboard_mask, nodeColor=<rgba 1 0.1406 0 1>}
  n154 = PreSkinnedNormal()
  NamedRerouteDeclaration(None=FunctionInput(Preview=VertexInterpolator(VS=@n154).PS) {bUsePreviewValueAsDefault=true, inputName=Pre-Skinned Local Normal, sortPriority=10}) {name=PreSkin_Normal, nodeColor=<rgba 0.8281 1 0 1>}
  NamedRerouteDeclaration(None=@n36.Normal) {name=Root_Normal, nodeColor=<rgba 1 0.2812 0 1>}
  NamedRerouteUsage()
  NamedRerouteUsage()
  NamedRerouteUsage()
  OneMinus(None=@n40.A)
  Saturate(None=Multiply(A=Add(A=@n39) {constB=0.5}))
  n155 = VectorParameter() {group=Wrap_CharlotteSummer, parameterName=CharS_BkG_Colors, defaultValue=<rgba 0.005 0.005 0.005 0.45>}
  SetMaterialAttributes!(MaterialAttributes=@n61, BaseColor=@n155.RGB, Roughness=@n155.RGB, Specular, Metallic, Normal, EmissiveColor) {attributeSetTypes [
    #2
    #3
    #5
    #4
    #1
    #6
   ]}
  TextureSample(UVs=@n21, Tex=NamedRerouteUsage(), Bias=@n3.MipLevel) {mipValueMode=TMVM_MipBias, samplerType=SAMPLERTYPE_Masks}
 ]}
