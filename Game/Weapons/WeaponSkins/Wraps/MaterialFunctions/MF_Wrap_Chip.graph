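# Weapon-wrap material function that applies an animated "chip" look to weapons, vehicles and mechs, behind the Wrap_Chip static switch.
# Side-projected UVs, with offset, tiling and rotation chosen by target type, sample base color/roughness, mask and gradient textures.
# The gradient, warped by mask alpha, scrolls a pulse texture, combined with a timed fade-out glow keyed to a TimeFrom-BP timestamp.
# The result feeds base color and emissive, blended over the input material by the customization mask.
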MaterialFunction /Game/Weapons/WeaponSkins/Wraps/MaterialFunctions/MF_Wrap_Chip
graph {nodes [
  n1 = PreSkinnedLocalBounds()
  n2 = Subtract(A=VertexInterpolator(VS=PreSkinnedPosition()).PS, B=@n1.Min)
  n3 = Subtract(A=@n1.Max, B=@n1.Max)
  n4 = Constant()
  n5 = FunctionInput(Preview=@n4) {inputName=isVehicle, inputType=FunctionInput_Scalar, sortPriority=5}
  n6 = MaterialFunctionCall(
   R=@n5
   G=FunctionInput(Preview=@n4) {bUsePreviewValueAsDefault=true, inputName=Side Project Vehicle, inputType=FunctionInput_Scalar}
   B=FunctionInput(Preview=@n4) {bUsePreviewValueAsDefault=true, inputName=isMech, inputType=FunctionInput_Scalar, sortPriority=4}
  ) {materialFunction=&AppendMany}
  n7 = MaterialFunctionCall(Color=@n6.RGBA) {materialFunction=&SplitComponents}
  n8 = VectorParameter() {group=Wrap_Chip, parameterName=Chip_BC_Offset_Weapon, defaultValue=<rgba -0.08 0.964 0 1>}
  n9 = LinearInterpolate(
   A=VectorParameter() {group=Wrap_Chip, parameterName=Chip_BC_Offset_Vehicle, defaultValue=<rgba 0 0 0 1>}.RGB
   B=VectorParameter() {group=Wrap_Chip, parameterName=Chip_BC_Offset_Vehicle_Side, defaultValue=<rgba 0 0 0 1>}.RGB
   Alpha=@n7.G
  )
  n10 = VectorParameter() {group=Wrap_Chip, parameterName=Chip_BC_TileRate_Weapon, defaultValue=<rgba 1 1 1 1>}
  n11 = Multiply(A=@n10.RGB, B=@n10.RGB)
  n12 = VectorParameter() {group=Wrap_Chip, parameterName=Chip_BC_TileRate_Vehicle, defaultValue=<rgba 1 1 1 1>}
  n13 = VectorParameter() {group=Wrap_Chip, parameterName=Chip_BC_TileRate_Vehicle_Side, defaultValue=<rgba 1 1 1 1>}
  n14 = MaterialFunctionCall(Color=@n6.RGBA) {materialFunction=&SplitComponents}
  n15 = LinearInterpolate(A=Multiply(A=@n12.RGB, B=@n12.RGB), B=Multiply(A=@n13.RGB, B=@n13.RGB), Alpha=@n14.G)
  n16 = VectorParameter() {group=Wrap_Chip, parameterName=Chip_BC_TileRate_Mech, defaultValue=<rgba 1 1 1 1>}
  n17 = MaterialFunctionCall!(
   For_Safe_Zones
   Normal_Map
   Caculated_Bounds=FunctionInput(Preview=Divide(A=@n2, B=@n3)) {bUsePreviewValueAsDefault=true, inputName=Calcated_Bounds, sortPriority=11}
   Pre_Skinned_Normals
   Is Vehicle=@n7.R
   Is VehicleSide=@n7.R
   Is Mech=@n7.R
   Vehicle_Rotation=ScalarParameter() {group=Wrap_Chip, parameterName=Chip_Rotation_Vehicle}
   UV_Rotation Bypass
   Mech_Rotation=ScalarParameter() {defaultValue=0.25, group=Wrap_Chip, parameterName=Chip_Rotation_Mech}
   Uvs_Offset=LinearInterpolate(
    A=LinearInterpolate(A=@n8.RGB, B=@n9, Alpha=@n7.R)
    B=VectorParameter() {group=Wrap_Chip, parameterName=Chip_BC_Offset_Mech, defaultValue=<rgba 0 0 0 1>}.RGB
    Alpha=@n7.B
   )
   Offset-after rotation
   Uvs_TileRate=LinearInterpolate(A=LinearInterpolate(A=@n11, B=@n15, Alpha=@n14.R), B=Multiply(A=@n16.RGB, B=@n16.RGB), Alpha=@n14.B)
  ) {materialFunction=&/MF_Projection_Side_Normal}
  n18 = TextureSampleParameter2D(UVs=@n17.UVs) {
   group=Wrap_Chip
   parameterName="Chip_Tex_Masks  A=Warp"
   samplerType=SAMPLERTYPE_Masks
   texture=&T_BlackMask
  }
  n19 = ScalarParameter() {group=Wrap_Chip, parameterName=Chip_Pulse_Proccesor_Brightness}
  n20 = Saturate(None=OneMinus(None=Multiply(A=@n18.G, B=@n19)))
  n21 = TextureSampleParameter2D(UVs=@n17.UVs) {group=Wrap_Chip, parameterName=Chip_Tex_Gradent, texture=&T_BlacksRGB}
  n22 = ScalarParameter() {defaultValue=0.25, group=Wrap_Chip, parameterName=Chip_Pulse_Warp_Amount}
  n23 = Add(
   A=Add(A=@n21.R, B=Multiply(A=@n18.A, B=@n22))
   B=Multiply(A=Time(), B=ScalarParameter() {defaultValue=0.5, group=Wrap_Chip, parameterName=Chip_Pulse_Speed})
  )
  n24 = Multiply(
   A=TextureSampleParameter2D(UVs=Frac(None=@n23)) {group=Wrap_Chip, parameterName=Chip_Tex_Pulse, texture=&T_BlacksRGB}.R
   B=VectorParameter() {group=Wrap_Chip, parameterName=Chip_Pulse_Color, defaultValue=<rgba 0.3121 0.3351 1 1>}.RGB
  )
  n25 = OneMinus(None=@n18.G)
  n26 = ScalarParameter() {group=Wrap_Chip, parameterName=Chip_Time_Warp_Amount}
  n27 = Subtract(A=Time(), B=ScalarParameter() {defaultValue=-9999999, group=Wrap_Chip, parameterName=TimeFrom-BP----------------, sortPriority=600})
  n28 = ScalarParameter() {defaultValue=0.8, group=Wrap_Chip, parameterName=Chip_Time_Duration, sortPriority=601}
  n29 = Saturate(None=Add(A=Multiply(A=@n18.A, B=@n26), B=Divide(A=@n27, B=@n28)))
  n30 = ScalarParameter() {group=Wrap_Chip, parameterName=Chip_Time_Curve, sortPriority=601}
  n31 = Multiply(
   A=LinearInterpolate(
    A=VectorParameter() {group=Wrap_Chip, parameterName=Chip_Time_Reactive_ColorB, defaultValue=<rgba 0.3121 0.3351 1 1>}.RGB
    B=VectorParameter() {group=Wrap_Chip, parameterName=Chip_Time_Reactive_Color, defaultValue=<rgba 0.3121 0.3351 1 1>}.RGB
    Alpha=@n21.R
   )
   B=OneMinus(None=Power(Base=@n29, Exp=@n30))
  )
  n32 = Multiply(A=@n18.R, B=Add(A=Multiply(A=@n20, B=@n24), B=Multiply(A=@n25, B=@n31)))
  n33 = Constant() {r=0.5}
  n34 = MaterialFunctionCall(Float2=@n17.UVs) {materialFunction=&BreakOutFloat2Components}
  n35 = SphereMask(A=@n33, B=@n34.R) {attenuationRadius=0.5}
  n36 = LinearInterpolate(A=@n35, Alpha=@n5)
  n37 = LinearInterpolate(B=@n32, Alpha=@n36)
  Add(A=@n37, B=LinearInterpolate(B=Multiply(A=@n18.R), Alpha=@n36))
  Constant()
  FunctionInput(Preview=VertexInterpolator(VS=PreSkinnedNormal()).PS) {bUsePreviewValueAsDefault=true, inputName=Pre-Skinned Local Normal, sortPriority=10}
  n38 = TextureSample() {samplerType=SAMPLERTYPE_Normal, texture=/Game/Weapons/FORT_Rifles/T_SCAR_N}
  n39 = GetMaterialAttributes(None=FunctionInput(Preview=SetMaterialAttributes!(MaterialAttributes, Normal=@n38.RGB) {attributeSetTypes [#1]}) {inputName=Material, inputType=FunctionInput_MaterialAttributes, sortPriority=-20})
  n40 = TextureSampleParameter2D(UVs=@n17.UVs) {group=Wrap_Chip, parameterName="Chip_Tex_BaseColor  A=Roughness", texture=&T_BlacksRGB}
  n41 = ScalarParameter() {group=Wrap_Chip, parameterName=Chip_Roughness_Low}
  n42 = ScalarParameter() {defaultValue=1, group=Wrap_Chip, parameterName=Chip_Roughness_High}
  n43 = SetMaterialAttributes!(
   MaterialAttributes=@n39.MaterialAttributes
   BaseColor=LinearInterpolate(B=Add(A=@n40.RGB, B=@n32), Alpha=@n36)
   Metallic=Constant()
   Specular=ScalarParameter() {defaultValue=0.5, group=Wrap_Chip, parameterName=Chip_Specular}
   Roughness=LinearInterpolate(
    A=ScalarParameter() {defaultValue=1, group=Wrap_Chip, parameterName=Chip_Roughness_Masked}
    B=LinearInterpolate(A=@n41, B=@n42, Alpha=@n40.A)
    Alpha=@n36
   )
   EmissiveColor=@n37
   Normal
  ) {attributeSetTypes [
    #2
    #3
    #4
    #5
    #6
    #1
   ]}
  n44 = Saturate(None=Add(
    A=FunctionInput() {bUsePreviewValueAsDefault=true, inputName="Customization Mask (_CM B)", inputType=FunctionInput_Scalar, sortPriority=24}
    B=ScalarParameter() {group=Wrap_Chip, parameterName=Chip_Remove_Customization Mask, sortPriority=-10}
   ))
  FunctionOutput(None=StaticSwitchParameter(True=BlendMaterialAttributes(A=@n39.MaterialAttributes, B=@n43, Alpha=@n44) {vertexAttributeBlendType=UseA}, False=@n39.MaterialAttributes) {group=Customization_Wrap, parameterName=Wrap_Chip}) {outputName=Result_1}
  MaterialFunctionCall() {materialFunction=&ScaleUVsByCenter}
  Multiply(A=@n35, B=SphereMask(A=@n34.G, B=@n33) {attenuationRadius=0.5})
  n45 = VectorParameter() {group=Wrap_Chip, parameterName=Chip_Pulse_Offset_Vehicle, defaultValue=<rgba 0.2 0 0 1>}
  n46 = VectorParameter() {group=Wrap_Chip, parameterName=Chip_Pulse_Offset_Vehicle_Side, defaultValue=<rgba 0.2 1 0 1>}
  n47 = LinearInterpolate(
   A=LinearInterpolate(
    A=VectorParameter() {group=Wrap_Chip, parameterName=Chip_Pulse_Offset_Weapon, defaultValue=<rgba 0.2 1 0 1>}.RGB
    B=LinearInterpolate(A=@n45.RGB, B=@n46.RGB, Alpha=@n7.G)
    Alpha=@n7.R
   )
   B=VectorParameter() {group=Wrap_Chip, parameterName=Chip_Pulse_Offset_Mech, defaultValue=<rgba -1 0 0 1>}.RGB
   Alpha=@n7.B
  )
  SphereMask(
   A=Add(A=@n17.UVs No Offsets/Tiling, B=ComponentMask(None=@n47) {g=true, r=true})
   B=Constant() {r=0.5}
   Radius=ScalarParameter() {defaultValue=0.6, group=Wrap_Chip, parameterName=Chip_Pulse_Radius}
   Hardness=ScalarParameter() {group=Wrap_Chip, parameterName=Chip_Pulse_Hardness}
  )
 ]}
