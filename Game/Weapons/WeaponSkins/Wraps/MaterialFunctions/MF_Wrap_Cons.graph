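# Material function for the Conscience weapon wrap, layering animated patterns and eyes over an incoming material by primary and secondary masks, behind a static switch.
# Side-projected UVs take offset, rotation and scale per item type (weapon, vehicle, character and others).
# Flow maps warp panning noise, which height-lerps against curve-atlas bands to build colors and roughness.
# Time-driven stepped offsets and curves animate blinking eyes, alongside Fresnel emissive and a separate accent layer.
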MaterialFunction /Game/Weapons/WeaponSkins/Wraps/MaterialFunctions/MF_Wrap_Cons
graph {nodes [
  n1 = NamedRerouteUsage()
  n2 = ScalarParameter() {defaultValue=1, group=Wrap_Conscience_Flow, parameterName=Cons_FlowMap2_Tiling}
  n3 = VectorParameter() {group=Wrap_Conscience_Flow, parameterName=Cons_FlowMap2_Speed, defaultValue=<rgba -0.056 0.03 0 1>}
  n4 = TextureSample(UVs=Panner(Coordinate=Multiply(A=@n1, B=@n2), Speed=ComponentMask(None=@n3.RGB) {g=true, r=true}) {speedY=0.02}, Tex=NamedRerouteUsage()) {samplerSource=SSM_Wrap_WorldGroupSettings}
  n5 = ComponentMask(None=@n4.RGB) {g=true, r=true}
  Add(A=@n5) {constB=-0.5}
  n6 = NamedRerouteUsage()
  n7 = TextureSample(UVs=Multiply(A=@n1, B=Constant2Vector() {g=-1, r=1}), Tex=@n6) {samplerSource=SSM_Wrap_WorldGroupSettings}
  n8 = OneMinus(None=@n7.RGB)
  n9 = TextureSample(UVs=@n1, Tex=@n6) {samplerSource=SSM_Wrap_WorldGroupSettings}
  n10 = ComponentMask(None=ComponentMask(None=LinearInterpolate(A=@n8, B=@n9.RGB, Alpha=@n9.RGB)) {g=true, r=true}) {g=true, r=true}
  Add(A=@n10) {constB=-0.5}
  n11 = VectorParameter() {group=Wrap_Conscience, parameterName=Cons_Color_A, defaultValue=<rgba 0.5 0.035 0.0921 1>}
  n12 = NamedRerouteUsage()
  n13 = Multiply(A=NamedRerouteUsage(), B=ScalarParameter() {defaultValue=0.15, group=Wrap_Conscience_Flow, parameterName=Cons_FlowMap_WarpAmount})
  n14 = Add(A=@n12, B=@n13)
  n15 = Multiply(A=NamedRerouteUsage(), B=ScalarParameter() {defaultValue=0.02, group=Wrap_Conscience_Flow, parameterName=Cons_FlowMap2_WarpAmount}) {constB=0}
  n16 = ComponentMask(None=Frac(None=Add(A=@n14, B=@n15))) {g=true}
  n17 = CurveAtlasRowParameter(CurveTime=@n16) {
   atlas=/Game/Weapons/WeaponSkins/Wraps/Textures/CruveAtlas_Wraps_01
   curve=/Game/Weapons/WeaponSkins/Wraps/Textures/0_ColorCurves/CC_Bars
   defaultValue=100
   group=Wrap_Conscience
   parameterName=Cons_CC_Msk
  }
  MaterialExpressionMaterialFunctionCall_7 = MaterialFunctionCall!(
   Index=ScalarParameter() {defaultValue=100, group=Wrap_Conscience, parameterName="Cons_CC_Msk  Index"}
   CCinput=@n17
   CCinput_Alpha=@n17
   UV=@n16
  ) {materialFunction=&/MF_Global_CC_Tex}
  n18 = Convert(None=@MaterialExpressionMaterialFunctionCall_7.4v) {
   nodeName=Break Float4
   convertInputs [{
     type=Vector4
     defaultValue=<rgba 0 0 0 1>
     expressionInput {
      expression=&self:MaterialExpressionMaterialFunctionCall_7
      inputName=None
      mask=0
      maskA=0
      maskB=0
      maskG=0
      maskR=0
      outputIndex=1
     }
    }]
   convertOutputs [{type=Scalar}, {type=Scalar}, {type=Scalar}, {type=Scalar}]
  }
  n19 = Multiply(
   A=NamedRerouteUsage()
   B=ComponentMask(None=VectorParameter() {group=Wrap_Conscience_Uvs, parameterName=Cons_TileRate_A_A, defaultValue=<rgba 6 3 0 1>}.RGB) {g=true, r=true}
  ) {constB=10}
  n20 = NamedRerouteUsage()
  n21 = Multiply(
   A=NamedRerouteUsage()
   B=ComponentMask(None=VectorParameter() {group=Wrap_Conscience_Uvs, parameterName=Cons_TileRate_A_B, defaultValue=<rgba 4 5 0 1>}.RGB) {g=true, r=true}
  )
  n22 = Saturate(None=Add(
    A=TextureSample(UVs=@n19, Tex=@n20) {samplerSource=SSM_Wrap_WorldGroupSettings}.R
    B=TextureSample(UVs=@n21, Tex=@n20) {samplerSource=SSM_Wrap_WorldGroupSettings}.R
   ))
  n23 = ScalarParameter() {defaultValue=2, group=Wrap_Conscience, parameterName=Cons_Contrast}
  n24 = Multiply(
   A=Ceil(None=@n18)
   B=MaterialFunctionCall(Transition Phase=@n18, Height Texture=@n22, Contrast=@n23) {materialFunction=&HeightLerp}.Alpha
  )
  n25 = VectorParameter() {group=Wrap_Conscience, parameterName=Cons_Color_B, defaultValue=<rgba 0.1413 0.0777 0.1347 1>}
  n26 = Multiply(
   A=NamedRerouteUsage()
   B=ComponentMask(None=VectorParameter() {group=Wrap_Conscience_Uvs, parameterName=Cons_TileRate_B_A, defaultValue=<rgba 3 1.5 0 1>}.RGB) {g=true, r=true}
  ) {constB=10}
  n27 = NamedRerouteUsage()
  n28 = Multiply(
   A=NamedRerouteUsage()
   B=ComponentMask(None=VectorParameter() {group=Wrap_Conscience_Uvs, parameterName=Cons_TileRate_B_B, defaultValue=<rgba 2 2.5 0 1>}.RGB) {g=true, r=true}
  )
  n29 = Saturate(None=Add(
    A=TextureSample(UVs=@n26, Tex=@n27) {samplerSource=SSM_Wrap_WorldGroupSettings}.R
    B=TextureSample(UVs=@n28, Tex=@n27) {samplerSource=SSM_Wrap_WorldGroupSettings}.R
   ))
  n30 = Multiply(
   A=Ceil(None=@n18)
   B=MaterialFunctionCall(Transition Phase=@n18, Height Texture=@n29, Contrast=@n23) {materialFunction=&HeightLerp}.Alpha
  )
  Add(A=Add(A=Multiply(A=@n11.RGB, B=@n24), B=Multiply(A=@n25.RGB, B=@n30)))
  n31 = TextureSample() {samplerType=SAMPLERTYPE_Masks, texture=/Game/Weapons/FORT_Rifles/Textures/T_Scar_CM}
  n32 = FunctionInput(Preview=@n31.R) {bUsePreviewValueAsDefault=true, inputName=Mask Primary, inputType=FunctionInput_Scalar, sortPriority=24, previewValue=<1 0 0 1>}
  n33 = FunctionInput(Preview=@n31.B) {bUsePreviewValueAsDefault=true, inputName=Mask Secondary, inputType=FunctionInput_Scalar, sortPriority=24, previewValue=<1 0 0 1>}
  n34 = BlendMaterialAttributes(A=NamedRerouteUsage(), B=NamedRerouteUsage(), Alpha=Saturate(None=Add(A=@n32, B=@n33))) {vertexAttributeBlendType=UseA}
  n35 = NamedRerouteUsage()
  n36 = NamedRerouteDeclaration(None=@n33) {name=Msk_Accent, nodeColor=<rgba 0 1 0.7656 1>}
  n37 = TextureSample() {samplerType=SAMPLERTYPE_Normal, texture=/Game/Weapons/FORT_Rifles/T_SCAR_N}
  n38 = Constant3Vector() {constant=<rgba 1 0 0 0>}
  n39 = GetMaterialAttributes(None=FunctionInput(Preview=SetMaterialAttributes!(MaterialAttributes, Normal=@n37.RGB, BaseColor=@n38) {attributeSetTypes [#1, #2]}) {inputName=Material, inputType=FunctionInput_MaterialAttributes}) {attributeGetTypes [
    #2
    #3
    #1
    #4
   ]}
  FunctionOutput(None=StaticSwitchParameter(
    True=BlendMaterialAttributes(A=@n34, B=@n35, Alpha=@n36) {vertexAttributeBlendType=UseA}
    False=NamedRerouteDeclaration(None=@n39.MaterialAttributes) {name=Root_Material, nodeColor=<rgba 0.7812 1 0 1>}
   ) {group=Customization_Wrap, parameterName=Wrap_Conscience})
  n40 = NamedRerouteUsage()
  n41 = NamedRerouteUsage()
  n42 = NamedRerouteUsage()
  LinearInterpolate(A=@n40, B=Add(A=@n40, B=LinearInterpolate(B=@n41, Alpha=@n42)), Alpha=NamedRerouteUsage())
  n43 = NamedRerouteUsage()
  n44 = TextureSample(UVs=NamedRerouteUsage(), Tex=@n43) {samplerSource=SSM_Wrap_WorldGroupSettings, samplerType=SAMPLERTYPE_Masks}
  n45 = Constant() {r=1}
  MaterialFunctionCall(Transition Phase=@n44.G, Height Texture=NamedRerouteUsage(), Contrast=@n45) {materialFunction=&HeightLerp}
  n46 = NamedRerouteUsage()
  n47 = Add(A=Time(), B=LinearInterpolate(Alpha=@n44.A) {constA=-2, constB=2})
  n48 = ComponentMask(None=VectorParameter() {group=Wrap_Conscience_Eye, parameterName=Cons_Eyes_Offse_Speed, defaultValue=<rgba 0.2 0.35 0 1>}.RGB) {g=true, r=true}
  n49 = AppendVector(A=Sine(None=Multiply(A=@n47, B=@n48)), B=Constant())
  n50 = ComponentMask(None=@n49) {r=true}
  n51 = Round(None=ComponentMask(None=@n49) {g=true})
  n52 = Multiply(
   A=AppendVector(A=Round(None=@n50), B=@n51)
   B=ScalarParameter() {defaultValue=0.01, group=Wrap_Conscience_Eye, parameterName=Cons_Eyes_Offse_Clamp}
  )
  n53 = NamedRerouteUsage()
  n54 = LinearInterpolate(
   A=VectorParameter() {parameterName=Cons_Eyes_ColorA, defaultValue=<rgba 0.0273 0.0232 0.0252 1>}.RGB
   B=TextureSample(UVs=Add(A=@n46, B=@n52), Tex=@n53) {samplerSource=SSM_Wrap_WorldGroupSettings, samplerType=SAMPLERTYPE_Masks}.RGB
   Alpha=MaterialFunctionCall(In=NamedRerouteDeclaration(None=@n44.B) {name=Eye_Msk, nodeColor=<rgba 1 0.9844 0 1>}, Contrast=Constant() {r=2}) {materialFunction=&CheapContrast}.Result
  )
  n55 = ScalarParameter() {defaultValue=100, group=Wrap_Conscience, parameterName="Cons_CC_Msk  Index"}
  n56 = Add(
   A=Multiply(A=NamedRerouteUsage()) {constB=0}
   B=Multiply(A=Time(), B=ScalarParameter() {defaultValue=0.1, group=Wrap_Conscience_Flow, parameterName=Cons_Eyes_Blinking_Speed})
  )
  n57 = Add(A=Frac(None=ComponentMask(None=@n56) {r=true}), B=Multiply(A=Add(A=@n44.A) {constB=-0.5}))
  n58 = CurveAtlasRowParameter(CurveTime=@n57) {
   atlas=&CruveAtlas_Wraps_01
   curve=&CC_Bars
   defaultValue=100
   group=Wrap_Conscience
   parameterName=Cons_CC_Msk
  }
  n59 = MaterialFunctionCall(
   Transition Phase=Saturate(None=MaterialFunctionCall(Index=@n55, CCinput_Alpha=@n58.A, UV=@n57) {materialFunction=&MF_Global_CC_Tex}.Alpha)
   Height Texture=TextureSample(UVs=NamedRerouteUsage(), Tex=@n53) {samplerSource=SSM_Wrap_WorldGroupSettings, samplerType=SAMPLERTYPE_Masks}.A
   Contrast=Constant() {r=5}
  ) {materialFunction=&HeightLerp}
  Multiply(A=@n54, B=@n59.Alpha)
  Multiply(A=@n18)
  n60 = NamedRerouteUsage()
  n61 = TextureSample(UVs=Add(A=NamedRerouteUsage(), B=Multiply(A=@n60) {constB=0.0025}), Tex=@n43) {samplerSource=SSM_Wrap_WorldGroupSettings, samplerType=SAMPLERTYPE_Masks}
  Multiply(A=NamedRerouteDeclaration(None=MaterialFunctionCall(In=@n61.G, Contrast=@n45) {materialFunction=&CheapContrast}.Result) {name=Eye_Msk_Background, nodeColor=<rgba 1 0.9844 0 1>})
  Multiply(A=@n54)
  Multiply(A=Add() {constB=-0.5}) {constB=2}
  Multiply(A=MaterialFunctionCall!(X range Y range, Gradient Height, ShowGrid?, Thickness, Color, "f(x)"=@n51) {materialFunction=/Engine/Functions/Engine_MaterialFunctions01/Debug/PlotFunctionOnGraph}."f(x) = y")
  Multiply(A=Constant3Vector() {constant=<rgba -1 -1 0 0>}, B=@n7.RGB)
  n62 = VectorParameter() {group=Wrap_Conscience_Accent, parameterName=Cons_Accent_BaseColor, defaultValue=<rgba 0.0365 0.0365 0.036 1>}
  n63 = VectorParameter() {group=Wrap_Conscience_Accent, parameterName=Cons_Accent_BaseColor_B, defaultValue=<rgba 0.224 0.224 0.2214 1>}
  n64 = ComponentMask(None=NamedRerouteUsage()) {r=true}
  n65 = SetMaterialAttributes!(
   MaterialAttributes=NamedRerouteUsage()
   BaseColor=LinearInterpolate(A=@n62.RGB, B=@n63.RGB, Alpha=Saturate(None=@n64))
   Roughness=ScalarParameter() {defaultValue=0.45, group=Wrap_Conscience_Accent, parameterName=Cons_Accent_Roughness}
   Specular=ScalarParameter() {defaultValue=0.25, group=Wrap_Conscience_Accent, parameterName=Cons_Accent_Spec}
   Metallic=ScalarParameter() {group=Wrap_Conscience_Accent, parameterName=Cons_Accent_Metalic}
   Normal
   EmissiveColor
  ) {attributeSetTypes [
    #2
    #3
    #5
    #4
    #1
    #6
   ]}
  NamedRerouteDeclaration(None=@n65) {name=Mat_Accent, nodeColor=<rgba 0 0.3906 1 1>}
  n66 = FunctionInput(Preview=@n31.G) {bUsePreviewValueAsDefault=true, inputName=Mask Crevise, inputType=FunctionInput_Scalar, sortPriority=24, previewValue=<1 0 0 1>}
  NamedRerouteDeclaration(None=@n66) {name=Msk_Curvature, nodeColor=<rgba 0 1 0.7422 1>}
  NamedRerouteDeclaration(None=TextureObjectParameter() {group=Wrap_Conscience_Flow, parameterName=Cons__Tex_Flow, texture=&T_BlacksRGB}) {name=Tex_Flow, nodeColor=<rgba 0.2656 0 1 1>}
  NamedRerouteDeclaration(None=Subtract(A=Multiply(A=@n5) {constB=2})) {name=FLowMap2, nodeColor=<rgba 0.9922 0 1 1>}
  NamedRerouteDeclaration(None=TextureObjectParameter() {group=Wrap_Conscience_Flow, parameterName=Cons__Tex_Flow2, texture=&T_BlacksRGB}) {name=Tex_Flow2, nodeColor=<rgba 0.2656 0 1 1>}
  NamedRerouteDeclaration(None=TextureObjectParameter() {group=Wrap_Conscience, parameterName=Cons__Tex_Msk, texture=&T_BlacksRGB}) {name=Tex_Msk, nodeColor=<rgba 0.2656 0 1 1>}
  NamedRerouteDeclaration(None=TextureObjectParameter() {group=Wrap_Conscience, parameterName=Cons__Tex_BC, texture=&T_BlacksRGB}) {name=Tex_Bc, nodeColor=<rgba 0.2656 0 1 1>}
  n67 = VectorParameter() {group=Wrap_Conscience_Uvs, parameterName=Cons_Noise_Speed_B, defaultValue=<rgba -0.025 -0.005 0 1>}
  n68 = Add(A=Panner(Coordinate=@n12, Speed=ComponentMask(None=@n67.RGB) {g=true, r=true}), B=@n13)
  n69 = Multiply(
   A=Add(A=NamedRerouteUsage()) {constB=-0.5}
   B=ScalarParameter() {defaultValue=-0.015, group=Wrap_Conscience_Flow, parameterName=Cons_Curvature_WarpAmount}
  )
  NamedRerouteDeclaration(None=Add(A=Add(A=@n68, B=@n15), B=@n69)) {name=Uv_Flow_Warp_A, nodeColor=<rgba 1 0.8906 0 1>}
  n70 = VectorParameter() {group=Wrap_Conscience_Uvs, parameterName=Cons_Noise_Speed_A, defaultValue=<rgba 0.025 0.005 0 1>}
  n71 = Add(A=Panner(Coordinate=@n12, Speed=ComponentMask(None=@n70.RGB) {g=true, r=true}), B=@n13)
  NamedRerouteDeclaration(None=Add(A=@n69, B=Add(A=@n71, B=@n15))) {name=Uv_Flow_Warp_B, nodeColor=<rgba 1 0.8906 0 1>}
  n72 = Power(
   Base=ComponentMask(None=NamedRerouteUsage()) {r=true}
   Exp=ScalarParameter() {defaultValue=2, group=Wrap_Conscience, parameterName=Cons_Color_B_Falloff}
  )
  n73 = NamedRerouteUsage()
  n74 = Divide(
   A=Add(A=ComponentMask(None=@n73) {r=true}, B=ScalarParameter() {defaultValue=-0.25, group=Wrap_Conscience, parameterName=Cons_Color_C_Offset}) {constB=-0.25}
   B=ScalarParameter() {defaultValue=0.25, group=Wrap_Conscience, parameterName=Cons_Color_C_Falloff}
  )
  n75 = NamedRerouteUsage()
  n76 = Multiply(
   A=Multiply(
    A=Ceil(None=@n18)
    B=MaterialFunctionCall(Transition Phase=@n18, Height Texture=@n29, Contrast=@n23) {materialFunction=&HeightLerp}.Alpha
   )
   B=@n75
  )
  n77 = LinearInterpolate(
   A=LinearInterpolate(
    A=VectorParameter() {group=Wrap_Conscience, parameterName=Cons_Color_B_A, defaultValue=<rgba 0.0296 0.0236 0.028 1>}.RGB
    B=@n25.RGB
    Alpha=Saturate(None=@n72)
   )
   B=LinearInterpolate(
    A=VectorParameter() {group=Wrap_Conscience, parameterName=Cons_Color_C_B, defaultValue=<rgba 0.05 0.0347 0.01 1>}.RGB
    B=VectorParameter() {group=Wrap_Conscience, parameterName=Cons_Color_C, defaultValue=<rgba 0.2195 0.1746 0.1022 1>}.RGB
    Alpha=Saturate(None=@n74)
   )
   Alpha=@n76
  )
  n78 = NamedRerouteUsage()
  n79 = Divide(
   A=Add(A=ComponentMask(None=@n78) {r=true}, B=ScalarParameter() {defaultValue=-0.25, group=Wrap_Conscience, parameterName=Cons_Color_A_Offset}) {constB=-0.25}
   B=ScalarParameter() {defaultValue=0.25, group=Wrap_Conscience, parameterName=Cons_Color_A_Falloff}
  )
  n80 = LinearInterpolate(
   A=VectorParameter() {group=Wrap_Conscience, parameterName=Cons_Color_A_B, defaultValue=<rgba 0.5271 0.0264 0.0885 1>}.RGB
   B=VectorParameter() {group=Wrap_Conscience, parameterName=Cons_Color_A, defaultValue=<rgba 0.5 0.035 0.0921 1>}.RGB
   Alpha=Multiply(A=NamedRerouteUsage(), B=Saturate(None=@n79))
  )
  n81 = Multiply(A=@n24, B=@n75)
  n82 = Multiply(A=NamedRerouteUsage(), B=NamedRerouteUsage())
  n83 = NamedRerouteUsage()
  n84 = LinearInterpolate(
   A=ScalarParameter() {defaultValue=1, group=Wrap_Conscience, parameterName=Cons_Roughness_B}
   B=ScalarParameter() {defaultValue=0.5, group=Wrap_Conscience, parameterName=Cons_Roughness_C}
   Alpha=@n76
  )
  n85 = ScalarParameter() {defaultValue=0.35, group=Wrap_Conscience, parameterName=Cons_Roughness_A}
  n86 = ScalarParameter() {defaultValue=0.3, group=Wrap_Conscience, parameterName=Cons_Roughness_EyeEdge}
  n87 = ScalarParameter() {defaultValue=0.15, group=Wrap_Conscience, parameterName=Cons_Roughness_Eye}
  n88 = Multiply(A=NamedRerouteUsage(), B=NamedRerouteUsage())
  n89 = NamedRerouteUsage()
  n90 = NamedRerouteUsage()
  n91 = Multiply(A=Multiply(A=NamedRerouteUsage(), B=NamedRerouteUsage()), B=NamedRerouteUsage())
  n92 = VectorParameter() {group=Wrap_Conscience, parameterName=Cons_Fresnel_Color, defaultValue=<rgba 0.5 0.035 0.0921 1>}
  n93 = Multiply(
   A=Fresnel(ExponentIn=ScalarParameter() {defaultValue=1, group=Wrap_Conscience, parameterName=Cons_Fresnel}) {baseReflectFraction=0}
   B=@n92.RGB
  )
  n94 = SetMaterialAttributes(
   MaterialAttributes=NamedRerouteUsage()
   BaseColor=LinearInterpolate(A=LinearInterpolate(A=@n77, B=@n80, Alpha=@n81), B=NamedRerouteUsage(), Alpha=Multiply(A=@n82, B=@n83)) {constB=0}
   Roughness=LinearInterpolate(A=LinearInterpolate(A=@n84, B=@n85, Alpha=@n81), B=LinearInterpolate(A=@n86, B=@n87, Alpha=@n88), Alpha=Multiply(A=@n89, B=@n90))
   Specular=LinearInterpolate(
    A=ScalarParameter() {group=Wrap_Conscience, parameterName=Cons_Spec}
    B=ScalarParameter() {defaultValue=0.5, group=Wrap_Conscience, parameterName=Cons_Spec_eye}
    Alpha=@n88
   )
   Metallic=ScalarParameter() {group=Wrap_Conscience, parameterName=Cons_Metalic}
   Normal=@n40
   EmissiveColor=Multiply(A=OneMinus(None=@n91), B=Multiply(A=@n93, B=@n92.A))
  ) {attributeSetTypes [
    #2
    #3
    #5
    #4
    #1
    #6
   ]}
  NamedRerouteDeclaration(None=@n94) {name=Mat_Base, nodeColor=<rgba 0 0.3906 1 1>}
  n95 = TextureObjectParameter() {
   group=Wrap_Conscience
   parameterName=Cons__Tex_Nrml
   samplerType=SAMPLERTYPE_Normal
   texture=&FlatNormalBlue
  }
  NamedRerouteDeclaration(None=@n95) {name=Tex_Nrl, nodeColor=<rgba 0.2656 0 1 1>}
  NamedRerouteDeclaration(None=Subtract(A=Multiply(A=@n10) {constB=2})) {name=FLowMap, nodeColor=<rgba 0.9922 0 1 1>}
  n96 = MaterialFunctionCall!(
   Normal_Map=NamedRerouteUsage()
   Offset Rt
   Pivot
   Caculated_Bounds=NamedRerouteUsage()
   Pre_Skinned_Normals=NamedRerouteUsage()
   is_State=NamedRerouteUsage()
   UV_Rotation=NamedRerouteUsage()
   Uvs_Offset=NamedRerouteUsage()
   Uvs_TileRate=NamedRerouteUsage()
   Uvs_Offset_Pivot
   For_Safe_Zones
   Rotation_Comp
   Clamp Y
   Clamp X
   Clamp Padding_Comp
   No_Rotation
  ) {materialFunction=&/MF_Projection_Side_Normal_V2}
  NamedRerouteDeclaration(None=@n96.Raw_Normal) {name=Normal_Position, nodeColor=<rgba 1 0.6563 0 1>}
  NamedRerouteDeclaration(None=@n22) {name=Panning Shapes, nodeColor=<rgba 1 0.7734 0 1>}
  NamedRerouteDeclaration(None=@n59.Alpha) {name=Eye_Blinking msk, nodeColor=<rgba 0 0.6719 1 1>}
  NamedRerouteDeclaration(None=@n54) {name=Eye_BC, nodeColor=<rgba 0.5 1 0 1>}
  n97 = MaterialFunctionCall(Color=Abs(None=NamedRerouteUsage())) {materialFunction=&SplitComponents}
  n98 = MaterialFunctionCall(Float4=NamedRerouteUsage()) {materialFunction=&BreakOutFloat4Components}
  n99 = LinearInterpolate(A=@n97.R, B=LinearInterpolate(A=@n97.B, B=@n97.B, Alpha=@n98.G), Alpha=@n98.R)
  NamedRerouteDeclaration(None=MaterialFunctionCall(
    In=LinearInterpolate(A=@n99, B=@n97.G, Alpha=@n98.B)
    Contrast=ScalarParameter() {defaultValue=5, group=Wrap_Conscience, parameterName=Cons_Blend_Contrast, sortPriority=0}
   ) {materialFunction=&CheapContrast}.Result) {name=Tri_Mask, nodeColor=<rgba 1 0.4688 0 1>}
  n100 = VectorParameter() {group=Wrap_Conscience_Uvs, parameterName=Cons_Offset_Vehicle, defaultValue=<rgba 0 0 0 1>}
  n101 = VectorParameter() {group=Wrap_Conscience_Uvs, parameterName=Cons_Offset_VehicleSide, defaultValue=<rgba 0 0 0 1>}
  n102 = MaterialFunctionCall(Float4=NamedRerouteUsage()) {materialFunction=&BreakOutFloat4Components}
  n103 = LinearInterpolate(
   A=LinearInterpolate(
    A=VectorParameter() {group=Wrap_Conscience_Uvs, parameterName=Cons_Offset_Weapon, defaultValue=<rgba 0 -0.16 0 1>}.RGB
    B=LinearInterpolate(A=@n100.RGB, B=@n101.RGB, Alpha=@n102.G)
    Alpha=@n102.R
   )
   B=LinearInterpolate(
    A=VectorParameter() {group=Wrap_Conscience_Uvs, parameterName=Cons_Offset_Mech, defaultValue=<rgba 0 0 0 1>}.RGB
    B=VectorParameter() {group=Wrap_Conscience_Uvs, parameterName=Cons_Offset_Character, defaultValue=<rgba 0 0 0 1>}.RGB
    Alpha=@n102.A
   )
   Alpha=@n102.B
  )
  n104 = VectorParameter() {group=Wrap_Conscience_Uvs, parameterName=Cons_Offset_Backpack, defaultValue=<rgba 0 0 0 1>}
  n105 = MaterialFunctionCall(Float2=NamedRerouteUsage()) {materialFunction=&BreakOutFloat2Components}
  NamedRerouteDeclaration(None=LinearInterpolate(
    A=LinearInterpolate(A=@n103, B=@n104.RGB, Alpha=@n105.R)
    B=VectorParameter() {group=Wrap_Conscience_Uvs, parameterName=Cons_Offset_Pickaxe, defaultValue=<rgba 0 0 0 1>}.RGB
    Alpha=@n105.G
   )) {name=UVs_Offset, nodeColor=<rgba 0 0.625 1 1>}
  n106 = ScalarParameter() {group=Wrap_Conscience_Uvs, parameterName=Cons_Rotation_Vehicle, sortPriority=150}
  n107 = ScalarParameter() {group=Wrap_Conscience_Uvs, parameterName=Cons_Rotation_Vehicle_Side, sortPriority=150}
  n108 = MaterialFunctionCall(Float4=NamedRerouteUsage()) {materialFunction=&BreakOutFloat4Components}
  n109 = LinearInterpolate(
   A=LinearInterpolate(
    A=ScalarParameter() {group=Wrap_Conscience_Uvs, parameterName=Cons_Rotation_Weapon, sortPriority=150}
    B=LinearInterpolate(A=@n106, B=@n107, Alpha=@n108.G)
    Alpha=@n108.R
   )
   B=LinearInterpolate(
    A=ScalarParameter() {group=Wrap_Conscience_Uvs, parameterName=Cons_Rotation_Mech, sortPriority=150}
    B=ScalarParameter() {group=Wrap_Conscience_Uvs, parameterName=Cons_Rotation_Character, sortPriority=150}
    Alpha=@n108.A
   )
   Alpha=@n108.B
  )
  n110 = ScalarParameter() {group=Wrap_Conscience_Uvs, parameterName=Cons_Rotation_Backpack, sortPriority=150}
  n111 = MaterialFunctionCall(Float2=NamedRerouteUsage()) {materialFunction=&BreakOutFloat2Components}
  NamedRerouteDeclaration(None=LinearInterpolate(
    A=LinearInterpolate(A=@n109, B=@n110, Alpha=@n111.R)
    B=ScalarParameter() {group=Wrap_Conscience_Uvs, parameterName=Cons_Rotation_Pickaxe, sortPriority=150}
    Alpha=@n111.G
   )) {name=UVs_Rotation, nodeColor=<rgba 0 0.7188 1 1>}
  n112 = ScalarParameter() {defaultValue=1, group=Wrap_Conscience_Uvs, parameterName=Cons_Scale_Vehicle, sortPriority=150}
  n113 = ScalarParameter() {defaultValue=1, group=Wrap_Conscience_Uvs, parameterName=Cons_Scale_Vehicle_Side, sortPriority=150}
  n114 = MaterialFunctionCall(Float4=NamedRerouteUsage()) {materialFunction=&BreakOutFloat4Components}
  n115 = LinearInterpolate(
   A=LinearInterpolate(
    A=ScalarParameter() {defaultValue=1, group=Wrap_Conscience_Uvs, parameterName=Cons_Scale_Weapon, sortPriority=150}
    B=LinearInterpolate(A=@n112, B=@n113, Alpha=@n114.G)
    Alpha=@n114.R
   )
   B=LinearInterpolate(
    A=ScalarParameter() {defaultValue=1, group=Wrap_Conscience_Uvs, parameterName=Cons_Scale_Mech, sortPriority=150}
    B=ScalarParameter() {defaultValue=1, group=Wrap_Conscience_Uvs, parameterName=Cons_Scale_Character, sortPriority=150}
    Alpha=@n114.A
   )
   Alpha=@n114.B
  )
  n116 = ScalarParameter() {defaultValue=1, group=Wrap_Conscience_Uvs, parameterName=Cons_Scale_Backpack, sortPriority=150}
  n117 = MaterialFunctionCall(Float2=NamedRerouteUsage()) {materialFunction=&BreakOutFloat2Components}
  NamedRerouteDeclaration(None=LinearInterpolate(
    A=LinearInterpolate(A=@n115, B=@n116, Alpha=@n117.R)
    B=ScalarParameter() {defaultValue=1, group=Wrap_Conscience_Uvs, parameterName=Cons_Scale_Pickaxe, sortPriority=150}
    Alpha=@n117.G
   )) {name=UVs_Scale, nodeColor=<rgba 0 0.7188 1 1>}
  NamedRerouteDeclaration(None=MaterialFunctionCall(UVs=@n96.UVs No Offsets/Tiling, Rotation Center=Constant() {r=0.5}, "Rotation Angle (0-1)"=NamedRerouteUsage()) {materialFunction=&CustomRotator}.Rotated Values) {name=Uvs_2, nodeColor=<rgba 1 0 0.2578 1>}
  n118 = Multiply(A=@n4.G) {constB=4}
  NamedRerouteDeclaration(None=Saturate(None=Power(Base=@n118))) {name=FLowMap2_Grad, nodeColor=<rgba 0.9922 0 1 1>}
  NamedRerouteDeclaration(None=FunctionInput(Preview=Constant4Vector().RGBA) {bUsePreviewValueAsDefault=true, inputName=IsState, inputType=FunctionInput_Vector4, sortPriority=4}) {name=IsState, nodeColor=<rgba 0.1497 1 0.1633 1>}
  NamedRerouteDeclaration(None=FunctionInput(Preview=Constant2Vector()) {bUsePreviewValueAsDefault=true, inputName=IsState_B, inputType=FunctionInput_Vector2, sortPriority=4}) {name=IsState_B, nodeColor=<rgba 0.1497 1 0.1633 1>}
  n119 = PreSkinnedLocalBounds()
  n120 = Subtract(A=VertexInterpolator(VS=PreSkinnedPosition()).PS, B=@n119.Min)
  n121 = MaterialFunctionCall(Color=Subtract(A=@n119.Max, B=@n119.Max)) {materialFunction=&SplitComponents}
  n122 = Max(A=Max(A=@n121.R, B=@n121.R), B=@n121.B)
  NamedRerouteDeclaration(None=FunctionInput(Preview=Divide(A=@n120, B=@n122)) {bUsePreviewValueAsDefault=true, inputName=Calcated_Bounds, sortPriority=11}) {name=Caclcated_Bounds, nodeColor=<rgba 0 0.9297 1 1>}
  NamedRerouteDeclaration(None=@n96.UVs) {name=Uvs, nodeColor=<rgba 1 0 0.2578 1>}
  n123 = PreSkinnedNormal()
  NamedRerouteDeclaration(None=FunctionInput(Preview=VertexInterpolator(VS=@n123).PS) {bUsePreviewValueAsDefault=true, inputName=Pre-Skinned Local Normal, sortPriority=10}) {name=PreSkin_Normal, nodeColor=<rgba 0.8281 1 0 1>}
  NamedRerouteDeclaration(None=@n39.Normal) {name=Root_Normal, nodeColor=<rgba 0.4062 0 1 1>}
  OneMinus(None=@n7.R)
  ScalarParameter() {defaultValue=0.1, group=Conscience, parameterName=OffsetSpeed}
  Sine(None=Multiply(A=Time(), B=ScalarParameter() {defaultValue=0.1, group=Conscience, parameterName=OffsetSpeed}))
  VectorParameter() {parameterName=Cons_Eyes_ColorB, defaultValue=<rgba 0.4179 0.2664 0.3864 1>}
 ]}
