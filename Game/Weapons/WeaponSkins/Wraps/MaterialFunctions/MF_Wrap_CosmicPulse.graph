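# Material function for the Cosmic Pulse wrap skin, blending an emissive galaxy look over the input material via customization masks behind a static switch.
# Camera-vector triplanar galaxy and star layers combine with a sine-pulsed fresnel rim.
# A tinted top/side-projected mask with panning, pulsing, warped UVs also distorts and multiplies a gradient, adding glow.
# The isState input selects weapon, vehicle, side-projected vehicle or mech projection settings.
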MaterialFunction /Game/Weapons/WeaponSkins/Wraps/MaterialFunctions/MF_Wrap_CosmicPulse
graph {nodes [
  n1 = FunctionInput() {inputName=Material, inputType=FunctionInput_MaterialAttributes, sortPriority=-10}
  n2 = FunctionInput() {
   bUsePreviewValueAsDefault=true
   desc=Base customization
   inputName="Customization Mask (_CM B Top left)"
   inputType=FunctionInput_Scalar
   sortPriority=24
  }
  n3 = FunctionInput() {
   bUsePreviewValueAsDefault=true
   desc=secondary customization
   inputName="Customization Mask 2 (_CM B Top Right)"
   inputType=FunctionInput_Scalar
   sortPriority=25
  }
  n4 = Add(A=@n3, B=@n2) {desc="Base + Secondary = exposed primary"}
  n5 = LinearInterpolate(
   A=LinearInterpolate(A=@n2, B=@n4, Alpha=ScalarParameter() {group=Wrap_CosmicPulse, parameterName="CP_BaseAndSecondaryMasked=ExposedPrimary", sortPriority=-99})
   B=Add(A=OneMinus(None=@n4) {desc=Primary}, B=@n2)
   Alpha=ScalarParameter() {group=Wrap_CosmicPulse, parameterName="CP_BaseAndPrimaryMasked=ExposedSecondary", sortPriority=-99}
  )
  n6 = Saturate(None=Subtract(A=NamedRerouteUsage(), B=NamedRerouteUsage()))
  n7 = GetMaterialAttributes(None=@n1) {attributeGetTypes [#1]}
  n8 = ScalarParameter() {defaultValue=1, group=Wrap_CosmicPulse, parameterName=CP_Galaxy_GlobalBrightness, sortPriority=2}
  n9 = Time()
  n10 = Multiply(A=@n9, B=Constant())
  n11 = ScalarParameter() {group=Wrap_CosmicPulse, parameterName=CP_Background Rotation}
  n12 = AppendVector(
   A=ScalarParameter() {defaultValue=0.02, group=Wrap_CosmicPulse, parameterName=CP_Galaxy Rotation Speed}
   B=ScalarParameter() {defaultValue=-0.02, group=Wrap_CosmicPulse, parameterName=CP_Galaxy Rotation Speed_1}
  )
  n13 = ScalarParameter() {defaultValue=0.1, group=Wrap_CosmicPulse, parameterName=CP_Galaxy Rotation Speed_2}
  n14 = MaterialFunctionCall!(
   CameraVector
   Rotation Speed=Multiply(A=LinearInterpolate(A=@n9, B=@n10, Alpha=@n11), B=AppendVector(A=@n12, B=@n13))
   Texture=TextureObjectParameter() {group=Wrap_CosmicPulse, parameterName=CP_GalaxyTexture, texture=&T_BlacksRGB}
   Tiling=ScalarParameter() {defaultValue=0.5, group=Wrap_CosmicPulse, parameterName=CP_GalaxyTiling}
   Offset=ScalarParameter() {defaultValue=0.012, group=Wrap_CosmicPulse, parameterName=CP_WarpIntensity}
   AxisFadeContrast=ScalarParameter() {defaultValue=6, group=Wrap_CosmicPulse, parameterName=CP_GalaxyAxisFade}
  ) {materialFunction=/Game/Weapons/WeaponSkins/MaterialFunction/TriplanarCameraVector_Rotated}
  n15 = Desaturation(None=@n14.Result, Fraction=ScalarParameter() {defaultValue=0.5, group=Wrap_CosmicPulse, parameterName=CP_Star Saturation})
  n16 = Multiply(A=@n15, B=VectorParameter() {group=Wrap_CosmicPulse, parameterName=CP_Star Brightness, defaultValue=<rgba 300 300 300 1>}.RGB)
  n17 = Time()
  n18 = ScalarParameter() {defaultValue=-0.125, group=Wrap_CosmicPulse, parameterName=CP_Small Star Rotation speed}
  n19 = TextureObjectParameter() {group=Wrap_CosmicPulse, parameterName=CP_StarsTexture, texture=&T_BlacksRGB}
  n20 = Time()
  n21 = VectorParameter() {group=Wrap_CosmicPulse, parameterName=CP_Small Star_Pan_Speed, defaultValue=<rgba 0 0 0 1>}
  n22 = ComponentMask(None=MaterialFunctionCall(
    Rotation Speed=Multiply(A=@n17, B=@n18)
    Texture=@n19
    Tiling=ScalarParameter() {defaultValue=5, group=Wrap_CosmicPulse, parameterName=CP_Small Star Tiling}
    Offset=Multiply(A=@n20, B=@n21.RGB)
   ) {materialFunction=&TriplanarCameraVector_Rotated}.Result) {b=true}
  n23 = Multiply(A=@n22, B=@n22)
  n24 = VectorParameter() {group=Wrap_CosmicPulse, parameterName=CP_Star Brightness_1, defaultValue=<rgba 950 950 950 1>}
  n25 = ComponentMask(None=MaterialFunctionCall!(
    CameraVector
    Texture=@n19
    Tiling=ScalarParameter() {defaultValue=2, group=Wrap_CosmicPulse, parameterName=CP_Bright Star Tiling}
    Offset
    AxisFadeContrast
   ) {materialFunction=/Engine/Functions/Engine_MaterialFunctions01/Texturing/TriplanarCameraVector}.Result) {r=true}
  n26 = Multiply(A=@n25, B=@n25)
  n27 = TextureSample(UVs=Panner(Coordinate=TextureCoordinate() {uTiling=2, vTiling=2}) {speedX=0.133, speedY=0.2}, Tex=@n19) {
   samplerSource=SSM_Wrap_WorldGroupSettings
   samplerType=SAMPLERTYPE_Masks
   texture=/Game/Characters/Player/Male/Medium/Bodies/M_Med_Soldier_04/Skins/Galaxy/Textures/Celestial_T_Stars
  }
  n28 = ScalarParameter() {defaultValue=8, group=Wrap_CosmicPulse, parameterName=CP_AccentRimBrightness, sortPriority=101}
  n29 = VectorParameter() {group=Wrap_CosmicPulse, parameterName=CP_EdgeFresnel, sortPriority=100, defaultValue=<rgba 5.0909 8.586 10 1>}
  n30 = Sine(None=Multiply(A=Time()) {constB=0.1}) {period=0.5}
  n31 = Clamp(
   None=Divide(A=Add(A=@n30))
   Min=ScalarParameter() {defaultValue=0.05, group=Wrap_CosmicPulse, parameterName=CP_FresnelPulseMin, sortPriority=103}
   Max=ScalarParameter() {defaultValue=1, group=Wrap_CosmicPulse, parameterName=CP_FresnelPulseMax, sortPriority=104}
  )
  n32 = Multiply(A=@n29.RGB, B=@n31)
  n33 = ScalarParameter() {defaultValue=1, group=Wrap_CosmicPulse, parameterName=CP_FresnelGlowIntensity, sortPriority=102}
  n34 = ScalarParameter() {defaultValue=15, group=Wrap_CosmicPulse, parameterName=CP_Galaxy_FresEX, sortPriority=99}
  n35 = LinearInterpolate(
   A=LinearInterpolate(A=LinearInterpolate(A=@n14.Result, B=@n16, Alpha=@n23), B=Multiply(A=@n15, B=@n24.RGB), Alpha=Multiply(A=@n26, B=@n27.G))
   B=LinearInterpolate(A=Multiply(A=@n28, B=@n14.Result), B=@n29.RGB, Alpha=Multiply(A=@n32, B=@n33))
   Alpha=Saturate(None=Fresnel(ExponentIn=@n34) {baseReflectFraction=0})
  )
  n36 = VectorParameter() {group=TopSideProjection_Emissive, parameterName=CP_Mask_Color, sortPriority=100, defaultValue=<rgba 5.0909 8.586 10 1>}
  n37 = FunctionInput(Preview=VertexInterpolator(VS=PreSkinnedNormal()).PS) {bUsePreviewValueAsDefault=true, inputName=Pre-Skinned Local Normal, sortPriority=10}
  n38 = ScalarParameter() {defaultValue=1, group=TopSideProjection, parameterName=CP_Blend_Contrast, sortPriority=-1}
  n39 = MaterialFunctionCall(In=Abs(None=ComponentMask(None=@n37) {b=true}), Contrast=@n38) {materialFunction=&CheapContrast}
  n40 = MaterialFunctionCall(In=Abs(None=ComponentMask(None=@n37) {g=true}), Contrast=@n38) {materialFunction=&CheapContrast}
  n41 = MaterialFunctionCall(In=Abs(None=ComponentMask(None=@n37) {r=true}), Contrast=@n38) {materialFunction=&CheapContrast}
  n42 = Add(A=Saturate(None=Subtract(A=@n39.Result, B=@n40.Result)), B=@n41.Result)
  n43 = ScalarParameter() {defaultValue=1, group=TopSideProjection_Emissive, parameterName=CP_AEmissive_alternate_Axis_Masking_Weapon, sortPriority=3}
  n44 = Add(A=Saturate(None=Subtract(A=@n39.Result, B=@n41.Result)), B=@n40.Result)
  n45 = ScalarParameter() {defaultValue=1, group=TopSideProjection_Emissive, parameterName=CP_AEmissive_alternate_Axis_Masking_Vehicle, sortPriority=3}
  n46 = FunctionInput(Preview=Constant()) {bUsePreviewValueAsDefault=true, inputName=isState, inputType=FunctionInput_Vector4, sortPriority=5}
  n47 = MaterialFunctionCall(Color=@n46) {materialFunction=&SplitComponents}
  n48 = LinearInterpolate(A=Add(A=Saturate(None=@n42), B=@n43), B=Add(A=Saturate(None=@n44), B=@n45), Alpha=@n47.R)
  n49 = FunctionInput(Preview=Constant3Vector() {constant=<rgba 1 0 0 0>}) {bUsePreviewValueAsDefault=true, inputName=ForwardAxis}
  n50 = TextureObjectParameter() {group=TopSideProjection, parameterName=CP_BaseColor_01, sortPriority=11, texture=&T_BlacksRGB}
  n51 = VectorParameter() {group=TopSideProjection_UVS, parameterName=CP-Rotation-Weapon-01---------------------, sortPriority=1, defaultValue=<rgba 0 0 -0.25 1>}
  n52 = VectorParameter() {group=TopSideProjection_UVS, parameterName=CP_Rotation_Vehicle_01, sortPriority=2, defaultValue=<rgba 0 0 0 1>}
  n53 = VectorParameter() {
   desc=must activate TS_Vehicle_Side_01 UVs
   group=TopSideProjection_UVS
   parameterName=CP_Rotation_Vehicle_Side_01
   sortPriority=2
   defaultValue=<rgba 0 0 0 1>
  }
  n54 = MaterialFunctionCall(Color=@n46) {materialFunction=&SplitComponents}
  n55 = LinearInterpolate(A=@n51.RGB, B=LinearInterpolate(A=@n52.RGB, B=@n53.RGB, Alpha=@n54.G), Alpha=@n54.R)
  n56 = VectorParameter() {group=TopSideProjection_UVS, parameterName=CP_Rotation_Mech_01, sortPriority=3, defaultValue=<rgba 0.25 0.25 0 1>}
  n57 = VectorParameter() {group=TopSideProjection_UVS, parameterName=CP_Offset_Weapon_01---------------------, sortPriority=10, defaultValue=<rgba 0 0 0 1>}
  n58 = VectorParameter() {group=TopSideProjection_UVS, parameterName=CP_Offset_Vehicle_01, sortPriority=11, defaultValue=<rgba 0 0 0 1>}
  n59 = VectorParameter() {
   desc=must activate TS_Vehicle_Side_01 UVs
   group=TopSideProjection_UVS
   parameterName=CP_Offset_Vehicle_Side_01
   sortPriority=11
   defaultValue=<rgba 0 0 0 1>
  }
  n60 = MaterialFunctionCall(Color=@n46) {materialFunction=&SplitComponents}
  n61 = LinearInterpolate(A=@n58.RGB, B=@n59.RGB, Alpha=@n60.G)
  n62 = VectorParameter() {group=TopSideProjection_UVS, parameterName=CP_Offset_Mech_01, sortPriority=12, defaultValue=<rgba 0 0 0 1>}
  n63 = PreSkinnedLocalBounds()
  n64 = Subtract(A=VertexInterpolator(VS=PreSkinnedPosition()).PS, B=@n63.Min)
  n65 = Subtract(A=@n63.Max, B=@n63.Max)
  n66 = Add(
   A=FunctionInput(Preview=Divide(A=@n64, B=@n65)) {bUsePreviewValueAsDefault=true, inputName=Calcated_Bounds, sortPriority=11}
   B=LinearInterpolate(A=LinearInterpolate(A=@n57.RGB, B=@n61, Alpha=@n60.R), B=@n62.RGB, Alpha=@n60.B)
  )
  n67 = Multiply(A=Time(), B=ScalarParameter() {defaultValue=0.1, group=TopSideProjection_UVS, parameterName=CP_Warp02_Pulse_Frequency, sortPriority=209})
  n68 = Divide(A=Add(A=Sine(None=@n67)))
  n69 = Multiply(
   A=@n68
   B=ScalarParameter() {defaultValue=0.5, group=TopSideProjection_UVS, parameterName=CP_Warp02_Pulse_Amplitude_Weapon, sortPriority=210}
  ) {constB=0.5}
  n70 = Multiply(
   A=@n68
   B=ScalarParameter() {defaultValue=0.5, group=TopSideProjection_UVS, parameterName=CP_Warp02_Pulse_Amplitude_Vehicle, sortPriority=210}
  ) {constB=0.5}
  n71 = MaterialFunctionCall(Color=@n46) {materialFunction=&SplitComponents}
  n72 = VectorParameter() {group=TopSideProjection_UVS, parameterName=CP_Panning_Weapon_01---------------------, sortPriority=40, defaultValue=<rgba 0 0 0 1>}
  n73 = VectorParameter() {group=TopSideProjection_UVS, parameterName=CP_Panning_Vehicle_01, sortPriority=41, defaultValue=<rgba 0 0 0 1>}
  n74 = VectorParameter() {
   desc=must activate TS_Vehicle_Side_01 UVs
   group=TopSideProjection_UVS
   parameterName=CP_Panning_Vehicle_Side_01
   sortPriority=41
   defaultValue=<rgba 0 0 0 1>
  }
  n75 = MaterialFunctionCall(Color=@n46) {materialFunction=&SplitComponents}
  n76 = LinearInterpolate(A=@n73.RGB, B=@n74.RGB, Alpha=@n75.G)
  n77 = VectorParameter() {group=TopSideProjection_UVS, parameterName=CP_Panning_Mech_01, sortPriority=42, defaultValue=<rgba 0 0 0 1>}
  n78 = Multiply(A=Time(), B=LinearInterpolate(A=LinearInterpolate(A=@n72.RGB, B=@n76, Alpha=@n75.R), B=@n77.RGB, Alpha=@n75.B))
  n79 = Add(
   A=Clamp(
    None=LinearInterpolate(A=@n69, B=@n70, Alpha=@n71.R)
    Min=ScalarParameter() {defaultValue=0.01, group=TopSideProjection_UVS, parameterName=CP_Warp02_Pulse_Sinewave_Clamp_Min, sortPriority=212}
    Max=ScalarParameter() {defaultValue=0.1, group=TopSideProjection_UVS, parameterName=CP_Warp02_Sinewave_Clamp_Max, sortPriority=213}
   )
   B=@n78
  )
  n80 = VectorParameter() {group=TopSideProjection_UVS, parameterName="CP_Scale_Weapon_01 A=uniform", sortPriority=30, defaultValue=<rgba 1 1 1 1>}
  n81 = Multiply(A=@n80.RGB, B=@n80.RGB)
  n82 = VectorParameter() {group=TopSideProjection_UVS, parameterName="CP_Scale_Vehicle_01 A=uniform", sortPriority=31, defaultValue=<rgba 1 1 1 1>}
  n83 = VectorParameter() {
   desc=must activate TS_Vehicle_Side_01 UVs
   group=TopSideProjection_UVS
   parameterName="CP_Scale_Vehicle_Side_01 A=uniform"
   sortPriority=31
   defaultValue=<rgba 1 1 1 1>
  }
  n84 = MaterialFunctionCall(Color=@n75.RGB) {materialFunction=&SplitComponents}
  n85 = LinearInterpolate(A=Multiply(A=@n82.RGB, B=@n82.RGB), B=Multiply(A=@n83.RGB, B=@n83.RGB), Alpha=@n84.G)
  n86 = VectorParameter() {group=TopSideProjection_UVS, parameterName="CP_Scale_Mech_01 A=uniform", defaultValue=<rgba 1 1 1 1>}
  n87 = Multiply(
   A=LinearInterpolate(
    A=Add(A=@n66, B=@n79)
    B=Add(A=@n66, B=@n78)
    Alpha=ScalarParameter() {group=Wrap_CosmicPulse, parameterName=CP_ForWrapRoll, sortPriority=213}
   )
   B=LinearInterpolate(A=LinearInterpolate(A=@n81, B=@n85, Alpha=@n84.R), B=Multiply(A=@n86.RGB, B=@n86.RGB), Alpha=@n84.B)
  )
  n88 = ScalarParameter() {defaultValue=1, group=TopSideProjection_UVS, parameterName=CP-WarpUVS-01-TileRate Weapon, sortPriority=53}
  n89 = ScalarParameter() {defaultValue=1, group=TopSideProjection_UVS, parameterName=CP-WarpUVS-01-TileRate Vehicle, sortPriority=54}
  n90 = VectorParameter() {group=TopSideProjection_UVS, parameterName=CP-WarpUVS-01-Anitmated Weapon, sortPriority=57, defaultValue=<rgba 0 0 0 1>}
  n91 = VectorParameter() {group=TopSideProjection_UVS, parameterName=CP-WarpUVS-01-Anitmated Vehicle, sortPriority=57, defaultValue=<rgba 0 0 0 1>}
  n92 = LinearInterpolate(A=@n90.RGB, B=@n91.RGB, Alpha=@n75.R)
  n93 = Add(
   A=Multiply(A=TextureCoordinate(), B=LinearInterpolate(A=@n88, B=@n89, Alpha=@n75.R))
   B=Multiply(A=Time() {period=1}, B=ComponentMask(None=@n92) {g=true, r=true})
  )
  n94 = Multiply(
   A=ChannelMaskParameter(None=TextureSample(UVs=@n93, Tex=@n19) {samplerType=SAMPLERTYPE_Masks}.RGBA) {group=TopSideProjection_UVS, parameterName=CP-WarpUVS-01_Channel, sortPriority=51}.R
   B=LinearInterpolate(
    A=ScalarParameter() {group=TopSideProjection_UVS, parameterName=CP-WarpUVS-01-Amount Weapon, sortPriority=55}
    B=ScalarParameter() {group=TopSideProjection_UVS, parameterName=CP-WarpUVS-01-Amount_Vehicle, sortPriority=56}
    Alpha=@n75.R
   )
  )
  n95 = ScalarParameter() {defaultValue=1, group=TopSideProjection, parameterName=CP_Blend_Contrast, sortPriority=-1}
  n96 = ScalarParameter() {defaultValue=2, group=TopSideProjection, parameterName=CP_Blend_Contrast_Mech, sortPriority=-1}
  n97 = VectorParameter() {group=TopSideProjection, parameterName="CP_Fill_alternate_Axis_Color_A=Roughness 01", sortPriority=3}
  n98 = MaterialFunctionCall(Color=@n46) {materialFunction=&SplitComponents}
  n99 = ChannelMaskParameter(None=MaterialFunctionCall!(
    IsCharacter
    IsStatsB
    Texture=@n50
    Rotation_Axis=LinearInterpolate(A=@n55, B=@n56.RGB, Alpha=@n54.B)
    Caculated_Bounds=Add(A=@n87, B=@n94)
    Pre_Skinned_Normals=@n37
    Blend_Contrast=LinearInterpolate(A=@n95, B=@n96)
    Fill_alternate_Axis_Opacity=ScalarParameter() {group=TopSideProjection, parameterName=CP_Fill_alternate_Axis_Opacity 01, sortPriority=2}
    "Fill_alternate_Axis_Color_A=Roughness"=AppendVector(A=@n97.RGB, B=@n97.RGB)
    Is_Vehicle=@n98.R
    Is_Mech=@n98.R
    Enable_Fill_Z_Weapon
    Fill_Z_Weapon_opacity
    Fill_Z_Weapon_Color
   ) {materialFunction=&/MF_TopsSideProjection_Base}.Combined) {group=TopSideProjection, parameterName=CP_Mask_TextureChannel, sortPriority=12}
  n100 = FunctionInput(Preview=VertexInterpolator(VS=PreSkinnedPosition()).PS) {bUsePreviewValueAsDefault=true, inputName=Pre-Skinned Local Position}
  n101 = MaterialFunctionCall(Color=@n46) {materialFunction=&SplitComponents}
  n102 = VectorParameter() {group=TopSideProjection_Emissive, parameterName=CP_Mask_WeaponOffset_ CE, sortPriority=180, defaultValue=<rgba 0 -1 0 1>}
  n103 = ScalarParameter() {defaultValue=0.5, group=TopSideProjection_Emissive, parameterName=CP_Mask_WeaponFalloff CE, sortPriority=180}
  n104 = VectorParameter() {group=TopSideProjection_Emissive, parameterName=CP_Mask_VehicleOffset_ CE, sortPriority=181, defaultValue=<rgba -1 0 0 1>}
  n105 = ScalarParameter() {defaultValue=0.5, group=TopSideProjection_Emissive, parameterName=CP_Mask_VehicleFalloff CE, sortPriority=181}
  n106 = VectorParameter() {group=TopSideProjection_Emissive, parameterName=CP_Mask_MechOffset_ CE, sortPriority=182, defaultValue=<rgba 0 0 -1 1>}
  n107 = ScalarParameter() {group=TopSideProjection_Emissive, parameterName=CP_Mask_MechFalloff CE, sortPriority=182}
  n108 = VectorParameter() {group=TopSideProjection_Emissive, parameterName=CP_Mask_UV_Spread_XY CE, sortPriority=183, defaultValue=<rgba 1 1 1 0>}
  n109 = VectorParameter() {group=TopSideProjection_Emissive, parameterName=CP_Mask_UV_Spread_Mech_XY CE, sortPriority=184, defaultValue=<rgba 1 1 1 0>}
  n110 = MaterialFunctionCall(Color=@n46) {materialFunction=&SplitComponents}
  n111 = LinearInterpolate(A=@n108.RGB, B=@n109.RGB, Alpha=@n110.B)
  n112 = LinearInterpolate(
   A=LinearInterpolate(
    A=VectorParameter() {group=TopSideProjection_Emissive, parameterName=CP_Mask_UV_AnimatedWeapon_XY CE, sortPriority=185}.RGB
    B=VectorParameter() {group=TopSideProjection_Emissive, parameterName=CP_Mask_UV_AnimatedVehicle_XY CE, sortPriority=186}.RGB
    Alpha=@n110.R
   )
   B=VectorParameter() {group=TopSideProjection_Emissive, parameterName=CP_Mask_UV_AnimatedMech_XY CE, sortPriority=187}.RGB
   Alpha=@n110.B
  )
  n113 = ScalarParameter() {defaultValue=0.1, group=TopSideProjection_Emissive, parameterName=CP_Mask_WarpAmount, sortPriority=190}
  n114 = ScalarParameter() {defaultValue=1, group=TopSideProjection_Emissive, parameterName=CP_Emissive01_Texture_Rotate_Mech, sortPriority=13}
  n115 = ScalarParameter() {defaultValue=1, group=TopSideProjection_Emissive, parameterName=CP_Emissive01_Texture_Rotate_Vehicle_Side, sortPriority=13}
  n116 = ScalarParameter() {defaultValue=1, group=TopSideProjection_Emissive, parameterName=CP_Emissive01_Texture_Rotate_Vehicle, sortPriority=13}
  n117 = TextureObjectParameter() {
   group=TopSideProjection_Emissive
   parameterName=CP_Gradient_01
   samplerType=SAMPLERTYPE_Masks
   sortPriority=11
   texture=&T_BlackMask
  }
  n118 = ChannelMaskParameter(None=TextureSample(
    UVs=MaterialFunctionCall!(
     ForwardAxis=@n49
     Bounds=@n99.R
     Pre-Skinned Local Position=@n100
     isVehicle=@n101.R
     isMech=@n101.R
     Side Project Vehicle=@n101.R
     WeaponOffset=@n102.RGB
     WeaponFalloff=@n103
     VehicleOffset=@n104.RGB
     VehicleFalloff=@n105
     MechOffset=@n106.RGB
     MechFalloff=@n107
     Invert_Gradent
     UV_Spread_XY=@n111
     UV_Animated_XY=@n112
     Warp_UVs=@n113
     Rotate_Mech=@n114
     Rotate_Vehicle_Side=@n115
     Rotate_Vehicle=@n116
    ) {materialFunction=&/MF_Wrap_Gradent_Creation}.Uvs
    Tex=@n117
   ) {samplerType=SAMPLERTYPE_Masks}.RGBA) {group=TopSideProjection_Emissive, parameterName=CP_Mask_Gradient_TextureChannel, sortPriority=120}
  n119 = Multiply(A=Saturate(None=LinearInterpolate(A=@n48, Alpha=@n47.B)), B=@n118.R)
  n120 = Multiply(A=@n99.R, B=@n119)
  n121 = Saturate(None=Multiply(A=@n120, B=Add(A=@n120, B=@n119)))
  n122 = SetMaterialAttributes!(
   MaterialAttributes=@n7.MaterialAttributes
   BaseColor=Constant()
   Specular=ScalarParameter() {group=Wrap_CosmicPulse, parameterName=CP_Galaxy Specular, sortPriority=0}
   Roughness=ScalarParameter() {defaultValue=1, group=Wrap_CosmicPulse, parameterName=CP_Galaxy Roughness, sortPriority=1}
   EmissiveColor=Add(A=Multiply(A=@n36.RGB, B=@n121), B=Multiply(A=@n8, B=@n35))
   Normal=@n7.MaterialAttributes
  ) {attributeSetTypes [
    #2
    #3
    #4
    #5
    #1
   ]}
  n123 = StaticSwitchParameter(True=BlendMaterialAttributes(A=@n1, B=@n122, Alpha=LinearInterpolate(A=@n5, B=@n2, Alpha=@n6)) {vertexAttributeBlendType=UseA}, False=@n1) {
   desc=Triplanar Camera Vector Projection - ambient emissive Topside projection with warp
   group=Customization_Wrap
   parameterName=Wrap_CosmicPulse
  }
  FunctionOutput(None=@n123) {outputName=Result_1}
  Multiply(A=@n69, B=@n78)
  NamedRerouteDeclaration(None=MaterialFunctionCall(Float4=@n46) {materialFunction=&BreakOutFloat4Components}.A) {name=IsCharacter, nodeColor=<rgba 0.973 1 0 1>}
  NamedRerouteDeclaration(None=FunctionInput() {bUsePreviewValueAsDefault=true, inputName=isCharAccent, inputType=FunctionInput_Scalar, sortPriority=78}) {name=isCharAccent, nodeColor=<rgba 0 0.9253 1 1>}
  n124 = Multiply(
   A=Clamp(
    None=Multiply(A=@n118.R, B=@n31)
    Min=ScalarParameter() {defaultValue=1, group=TopSideProjection_Emissive, parameterName=CP_Mask_Gradent_Clamp Min, sortPriority=151}
    Max=ScalarParameter() {defaultValue=1, group=TopSideProjection_Emissive, parameterName=CPS_Mask_Gradent_Clamp Max, sortPriority=151}
   ) {maxDefault=0.99, minDefault=0.01}
   B=ScalarParameter() {defaultValue=1, group=TopSideProjection_Emissive, parameterName=CP_Mask_Gradent_Opacity_CE, sortPriority=149}
  )
  Saturate(None=MaterialFunctionCall(
    Transition Phase=Saturate(None=@n124)
    Height Texture=@n99.R
    Contrast=ScalarParameter() {defaultValue=1, group=TopSideProjection_Emissive, parameterName=CP_Mask_HeightLerp_Contrast, sortPriority=145}
   ) {materialFunction=&HeightLerp}.Alpha)
 ]}
