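# Material function for the "Defect" weapon wrap skin, blended over the input material through primary and secondary masks behind the Wrap_Defect static switch.
# Six texture layers project on SafeZone virtual planes, offset by a sine-animated height to fake depth.
# These combine with scrolling noise, bars, curvature and Fresnel into the emissive.
# IsState inputs select scale, height and Fresnel exponent per item type, such as weapon, vehicle or pickaxe.
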MaterialFunction /Game/Weapons/WeaponSkins/Wraps/MaterialFunctions/MF_Wrap_Defect
graph {nodes [
  n1 = ScalarParameter() {defaultValue=1, group=Wrap_Defect_Layers, parameterName=Defect_Layers_Opacty_0}
  n2 = MaterialFunctionCall!(
   Camera _pos
   Axis
   Scale
   Scale_Multplyer
   Object-SafeZone-Virtual-Pivot=NamedRerouteUsage()
   Object-SafeZone-Virtual-CH-Switch=NamedRerouteUsage()
   SinglePlane Normal Axis=NamedRerouteUsage()
   SinglePlane X Axis=NamedRerouteUsage()
   PlaneCenter_distortion=Multiply(A=NamedRerouteUsage(), B=ScalarParameter() {group=Wrap_Defect_Layers, parameterName=Defect_Layers_LAYER_0})
  ) {materialFunction=&/MF_SafeZone_VirtualPlane}
  n3 = NamedRerouteUsage()
  n4 = Saturate(None=LinearInterpolate(A=@n2.Single_plane_UVs, B=@n2.Single_plane_UVs, Alpha=@n3))
  n5 = Multiply(A=NamedRerouteUsage(), B=@n4)
  n6 = TextureSample(UVs=@n5, Tex=NamedRerouteUsage())
  Add(A=@n1, B=@n6.A)
  n7 = MaterialFunctionCall(
   Object-SafeZone-Virtual-Pivot=NamedRerouteUsage()
   Object-SafeZone-Virtual-CH-Switch=NamedRerouteUsage()
   SinglePlane Normal Axis=NamedRerouteUsage()
   SinglePlane X Axis=NamedRerouteUsage()
   PlaneCenter_distortion=Multiply(A=NamedRerouteUsage(), B=ScalarParameter() {defaultValue=0.5, group=Wrap_Defect_Layers, parameterName=Defect_Layers_LAYER_2})
  ) {materialFunction=&MF_SafeZone_VirtualPlane}
  n8 = NamedRerouteUsage()
  n9 = Saturate(None=LinearInterpolate(A=@n7.Single_plane_UVs, B=@n7.Single_plane_UVs, Alpha=@n8))
  n10 = Multiply(A=NamedRerouteUsage(), B=@n9)
  n11 = TextureSample(UVs=@n10, Tex=NamedRerouteUsage())
  Add(A=Constant() {r=0.75}, B=@n11.A)
  n12 = MaterialFunctionCall(
   Object-SafeZone-Virtual-Pivot=NamedRerouteUsage()
   Object-SafeZone-Virtual-CH-Switch=NamedRerouteUsage()
   SinglePlane Normal Axis=NamedRerouteUsage()
   SinglePlane X Axis=NamedRerouteUsage()
   PlaneCenter_distortion=Multiply(A=NamedRerouteUsage(), B=ScalarParameter() {defaultValue=0.75, group=Wrap_Defect_Layers, parameterName=Defect_Layers_LAYER_3})
  ) {materialFunction=&MF_SafeZone_VirtualPlane}
  n13 = NamedRerouteUsage()
  n14 = Saturate(None=LinearInterpolate(A=@n12.Single_plane_UVs, B=@n12.Single_plane_UVs, Alpha=@n13))
  n15 = Multiply(A=NamedRerouteUsage(), B=@n14)
  n16 = TextureSample(UVs=@n15, Tex=NamedRerouteUsage())
  Add(A=Constant() {r=0.6}, B=@n16.A)
  n17 = MaterialFunctionCall(
   Object-SafeZone-Virtual-Pivot=NamedRerouteUsage()
   Object-SafeZone-Virtual-CH-Switch=NamedRerouteUsage()
   SinglePlane Normal Axis=NamedRerouteUsage()
   SinglePlane X Axis=NamedRerouteUsage()
   PlaneCenter_distortion=Multiply(A=NamedRerouteUsage(), B=ScalarParameter() {defaultValue=1, group=Wrap_Defect_Layers, parameterName=Defect_Layers_LAYER_4})
  ) {materialFunction=&MF_SafeZone_VirtualPlane}
  n18 = NamedRerouteUsage()
  n19 = Saturate(None=LinearInterpolate(A=@n17.Single_plane_UVs, B=@n17.Single_plane_UVs, Alpha=@n18))
  n20 = Multiply(A=NamedRerouteUsage(), B=@n19)
  n21 = TextureSample(UVs=@n20, Tex=NamedRerouteUsage())
  Add(A=Constant() {r=0.5}, B=@n21.A)
  n22 = MaterialFunctionCall(
   Object-SafeZone-Virtual-Pivot=NamedRerouteUsage()
   Object-SafeZone-Virtual-CH-Switch=NamedRerouteUsage()
   SinglePlane Normal Axis=NamedRerouteUsage()
   SinglePlane X Axis=NamedRerouteUsage()
   PlaneCenter_distortion=Multiply(A=NamedRerouteUsage(), B=ScalarParameter() {defaultValue=1.25, group=Wrap_Defect_Layers, parameterName=Defect_Layers_LAYER_5})
  ) {materialFunction=&MF_SafeZone_VirtualPlane}
  n23 = NamedRerouteUsage()
  n24 = Saturate(None=LinearInterpolate(A=@n22.Single_plane_UVs, B=@n22.Single_plane_UVs, Alpha=@n23))
  n25 = Multiply(A=NamedRerouteUsage(), B=@n24)
  n26 = TextureSample(UVs=@n25, Tex=NamedRerouteUsage())
  Add(A=Constant() {r=0.4}, B=@n26.A)
  Add(A=Constant() {r=0.85}) {constB=0}
  Constant3Vector() {constant=<rgba 1 1 1 0>}
  Constant() {r=0.5}
  Constant() {r=0.25}
  Constant() {r=0.5}
  Constant() {r=1}
  n27 = TextureSample() {samplerType=SAMPLERTYPE_Masks, texture=/Game/Weapons/FORT_Rifles/Textures/T_Scar_CM}
  n28 = FunctionInput(Preview=@n27.R) {bUsePreviewValueAsDefault=true, inputName=Mask Primary, inputType=FunctionInput_Scalar, sortPriority=24, previewValue=<1 0 0 1>}
  n29 = FunctionInput(Preview=@n27.B) {bUsePreviewValueAsDefault=true, inputName=Mask Secondary, inputType=FunctionInput_Scalar, sortPriority=24, previewValue=<1 0 0 1>}
  n30 = BlendMaterialAttributes(A=NamedRerouteUsage(), B=NamedRerouteUsage(), Alpha=Saturate(None=Add(A=@n28, B=@n29))) {vertexAttributeBlendType=UseA}
  n31 = NamedRerouteUsage()
  n32 = NamedRerouteDeclaration(None=@n29) {name=Msk_Accent, nodeColor=<rgba 0 1 0.7656 1>}
  n33 = TextureSample() {samplerType=SAMPLERTYPE_Normal, texture=/Game/Weapons/FORT_Rifles/T_SCAR_N}
  n34 = Constant3Vector() {constant=<rgba 1 0 0 0>}
  n35 = GetMaterialAttributes(None=FunctionInput(Preview=SetMaterialAttributes!(MaterialAttributes, Normal=@n33.RGB, BaseColor=@n34) {attributeSetTypes [#1, #2]}) {inputName=Material, inputType=FunctionInput_MaterialAttributes}) {attributeGetTypes [
    #2
    #3
    #1
    #4
   ]}
  FunctionOutput(None=StaticSwitchParameter(
    True=BlendMaterialAttributes(A=@n30, B=@n31, Alpha=@n32) {vertexAttributeBlendType=UseA}
    False=NamedRerouteDeclaration(None=@n35.MaterialAttributes) {name=Root_Material, nodeColor=<rgba 0.7812 1 0 1>}
   ) {group=Customization_Wrap, parameterName=Wrap_Defect})
  n36 = MaterialFunctionCall(
   Object-SafeZone-Virtual-Pivot=NamedRerouteUsage()
   Object-SafeZone-Virtual-CH-Switch=NamedRerouteUsage()
   SinglePlane Normal Axis=NamedRerouteUsage()
   SinglePlane X Axis=NamedRerouteUsage()
   PlaneCenter_distortion=Multiply(A=NamedRerouteUsage(), B=ScalarParameter() {defaultValue=1.5, parameterName=Defect_Layers_LAYER_6})
  ) {materialFunction=&MF_SafeZone_VirtualPlane}
  n37 = TextureSample(UVs=Multiply(A=NamedRerouteUsage(), B=Saturate(None=@n36.Single_plane_UVs)), Tex=NamedRerouteUsage())
  n38 = Power(Base=@n37.R, Exp=Constant()) {constExponent=0.9}
  n39 = OneMinus(None=Round(None=Saturate(None=@n38)))
  n40 = Round(None=Saturate(None=OneMinus(None=@n37.R)))
  LinearInterpolate(B=Add(A=Constant() {r=0.3}, B=@n37.A), Alpha=Multiply(A=Multiply(A=@n39, B=@n40) {constA=1}, B=@n36.Single_plane_Masking))
  n41 = Constant3Vector() {constant=<rgba 0 1 0 1>}
  n42 = ScalarParameter() {group=Weapon Specific Wrap Adjustments, parameterName=DO NOT OVERRIDE-Pickaxe-Position-BP-Switch, sortPriority=111}
  n43 = LinearInterpolate(A=Constant3Vector() {constant=<rgba 1 0 0 1>}, B=@n41, Alpha=@n42)
  n44 = MaterialFunctionCall(Float4=NamedRerouteUsage()) {materialFunction=&BreakOutFloat4Components}
  n45 = LinearInterpolate(A=Constant3Vector() {constant=<rgba 0 0 1 1>}, B=@n41, Alpha=@n44.G)
  n46 = VectorParameter() {group=Wrap_Defect_Layers, parameterName=Defect_Character_Plane_Normal, sortPriority=0, defaultValue=<rgba 0 1 0 1>}
  n47 = Transform(None=LinearInterpolate(A=LinearInterpolate(A=@n43, B=@n45, Alpha=@n44.R), B=LinearInterpolate(A=@n46.RGB, B=@n46.RGB, Alpha=@n44.A), Alpha=@n44.B)) {transformSourceType=TRANSFORMSOURCE_Local}
  n48 = Normalize(VectorInput=@n47)
  n49 = MaterialFunctionCall(Float4=NamedRerouteUsage()) {materialFunction=&BreakOutFloat4Components}
  n50 = LinearInterpolate(
   A=LinearInterpolate(A=Constant3Vector() {constant=<rgba 0 1 0 1>}, B=Constant3Vector() {constant=<rgba 0 0 1 1>}, Alpha=@n42)
   B=LinearInterpolate(A=Constant3Vector() {constant=<rgba 0 -1 0 1>}, B=Constant3Vector() {constant=<rgba -1 0 0 1>}, Alpha=@n49.G)
   Alpha=@n49.R
  )
  n51 = Constant3Vector() {constant=<rgba -1 0 0 1>}
  n52 = LinearInterpolate(A=@n51, B=@n51, Alpha=@n49.A)
  n53 = Normalize(VectorInput=Transform(None=LinearInterpolate(A=@n50, B=@n52, Alpha=@n49.B)) {transformSourceType=TRANSFORMSOURCE_Local})
  n54 = MaterialFunctionCall(SinglePlane Normal Axis=Normalize(VectorInput=@n48), SinglePlane X Axis=Normalize(VectorInput=@n53)) {materialFunction=&MF_SafeZone_VirtualPlane}
  Multiply(
   A=TextureSample(UVs=@n54.Single_plane_UVs) {texture=/Game/Weapons/WeaponSkins/Wraps/Textures/UVCheaker_Clamped}.RGB
   B=@n54.Single_plane_Masking
  )
  Multiply(A=MaterialFunctionCall() {materialFunction=&Scalar_Time_of_Day}.Result)
  n55 = MaterialFunctionCall(
   Object-SafeZone-Virtual-Pivot=NamedRerouteUsage()
   Object-SafeZone-Virtual-CH-Switch=NamedRerouteUsage()
   SinglePlane Normal Axis=NamedRerouteUsage()
   SinglePlane X Axis=NamedRerouteUsage()
   PlaneCenter_distortion=Multiply(A=NamedRerouteUsage(), B=ScalarParameter() {defaultValue=0.25, group=Wrap_Defect_Layers, parameterName=Defect_Layers_LAYER_1})
  ) {materialFunction=&MF_SafeZone_VirtualPlane}
  n56 = NamedRerouteUsage()
  n57 = Saturate(None=LinearInterpolate(A=@n55.Single_plane_UVs, B=@n55.Single_plane_UVs, Alpha=@n56))
  n58 = Multiply(A=NamedRerouteUsage(), B=@n57)
  n59 = TextureSample(UVs=@n58, Tex=NamedRerouteUsage())
  n60 = Constant() {r=70}
  n61 = Round(None=Saturate(None=Power(Base=@n59.R, Exp=@n60) {constExponent=70}))
  n62 = Constant() {r=3}
  n63 = Saturate(None=Power(Base=@n59.R, Exp=@n62) {constExponent=5})
  n64 = LinearInterpolate(A=@n55.Single_plane_Masking, Alpha=@n56)
  Multiply(A=Multiply(A=OneMinus(None=@n61), B=Round(None=@n63)), B=@n64)
  n65 = Round(None=Saturate(None=Power(Base=@n11.R, Exp=@n62) {constExponent=5}))
  n66 = Constant() {r=1}
  n67 = Saturate(None=Power(Base=@n11.R, Exp=@n66) {constExponent=3})
  n68 = LinearInterpolate(A=@n7.Single_plane_Masking, Alpha=@n8)
  Multiply(A=Multiply(A=OneMinus(None=@n65), B=Round(None=@n67)) {constA=1}, B=@n68)
  n69 = Constant() {r=0.7}
  n70 = Round(None=Saturate(None=Power(Base=@n21.R, Exp=@n69) {constExponent=1.5}))
  n71 = Constant() {r=0.25}
  n72 = Saturate(None=Power(Base=@n21.R, Exp=@n71) {constExponent=1.25})
  n73 = LinearInterpolate(A=@n17.Single_plane_Masking, Alpha=@n18)
  Multiply(A=Multiply(A=OneMinus(None=@n70), B=Round(None=@n72)) {constA=1}, B=@n73)
  n74 = Round(None=Saturate(None=Power(Base=@n26.R, Exp=@n71) {constExponent=1.25}))
  n75 = Saturate(None=OneMinus(None=@n26.R))
  n76 = LinearInterpolate(A=@n22.Single_plane_Masking, Alpha=@n23)
  Multiply(A=Multiply(A=OneMinus(None=@n74), B=Round(None=@n75)) {constA=1}, B=@n76)
  n77 = Round(None=Saturate(None=Power(Base=@n16.R, Exp=@n66) {constExponent=3}))
  n78 = Saturate(None=Power(Base=@n16.R, Exp=@n69) {constExponent=1.5})
  n79 = LinearInterpolate(A=@n12.Single_plane_Masking, Alpha=@n13)
  Multiply(A=Multiply(A=OneMinus(None=@n77), B=Round(None=@n78)) {constA=1}, B=@n79)
  n80 = VectorParameter() {group=Wrap_Defect_Accent, parameterName=Defect_Accent_BaseColor, defaultValue=<rgba 0.015 0.0156 0.0153 1>}
  n81 = SetMaterialAttributes!(
   MaterialAttributes=NamedRerouteUsage()
   BaseColor=@n80.RGB
   Roughness=ScalarParameter() {defaultValue=0.55, group=Wrap_Defect_Accent, parameterName=Defect_Accent_Roughness}
   Specular=ScalarParameter() {defaultValue=0.3, group=Wrap_Defect_Accent, parameterName=Defect_Accent_Spec}
   Metallic=ScalarParameter() {group=Wrap_Defect_Accent, parameterName=Defect_Accent_Metalic}
   Normal
   EmissiveColor
  ) {attributeSetTypes [
    #2
    #3
    #5
    #4
    #1
    #6
   ]}
  NamedRerouteDeclaration(None=@n81) {name=Mat_Accent, nodeColor=<rgba 0 0.3906 1 1>}
  n82 = FunctionInput(Preview=@n27.G) {bUsePreviewValueAsDefault=true, inputName=Mask Crevise, inputType=FunctionInput_Scalar, sortPriority=24, previewValue=<1 0 0 1>}
  NamedRerouteDeclaration(None=@n82) {name=Msk_Curvature, nodeColor=<rgba 0 1 0.7422 1>}
  n83 = VectorParameter() {group=Wrap_Defect, parameterName="Defect_VP_Offset_A=Scale"}
  n84 = FunctionInput(Preview=Constant4Vector() {constant=<rgba 0 0 5 -1>}.RGBA) {bUsePreviewValueAsDefault=true, inputName=Object-SafeZone-Virtual-Pivot, inputType=FunctionInput_Vector4, sortPriority=105}
  n85 = Add(A=AppendVector(A=@n83.RGB, B=@n83.RGB), B=@n84)
  n86 = LinearInterpolate(Alpha=NamedRerouteUsage()) {constA=1, constB=-1}
  n87 = ComponentMask(None=@n85) {a=true}
  NamedRerouteDeclaration(None=AppendVector(A=ComponentMask(None=@n85) {b=true, g=true, r=true}, B=Multiply(A=@n86, B=@n87))) {name=SafeZone VP, nodeColor=<rgba 0.5 1 0 1>}
  NamedRerouteDeclaration() {name=UVs_Scale_V1, nodeColor=<rgba 0 0.7188 1 1>}
  n88 = FunctionInput(Preview=Constant()) {bUsePreviewValueAsDefault=true, inputName=Object-SafeZone-VP-Character-PreView-Switch, inputType=FunctionInput_Scalar, sortPriority=105}
  NamedRerouteDeclaration(None=@n88) {name=SafeZone Character-Switch, nodeColor=<rgba 1 0.9375 0 1>}
  NamedRerouteDeclaration(None=@n48) {name=SinglePlane_Normal, nodeColor=<rgba 0 0.9297 1 1>}
  NamedRerouteDeclaration(None=@n53) {name=SinglePlane_X_Axis, nodeColor=<rgba 0 0.9297 1 1>}
  NamedRerouteDeclaration(None=TextureObjectParameter() {group=Wrap_Defect_Tex, parameterName=Defect__Tex_Msk_A, texture=&T_BlacksRGB}) {name=Tex_Msk_A, nodeColor=<rgba 0.8984 1 0 1>}
  NamedRerouteDeclaration(None=@n47) {name=Constraint Axis, nodeColor=<rgba 0 0.9297 1 1>}
  n89 = Fresnel(
   ExponentIn=NamedRerouteUsage()
   BaseReflectFractionIn=ScalarParameter() {group=Wrap_Defect, parameterName=Defect_Fresnel_BaseREf, sortPriority=100}
  )
  n90 = VectorParameter() {group=Wrap_Defect, parameterName=Defect_Fresnel_Color, defaultValue=<rgba 1 0.4205 0 1>}
  n91 = NamedRerouteUsage()
  n92 = Saturate(None=Multiply(A=Add(A=@n91) {constB=-0.5}) {constB=2})
  n93 = Multiply(A=VectorParameter() {group=Wrap_Defect, parameterName=Defect_Color_Curvature, defaultValue=<rgba 1 0.4205 0 1>}.RGB, B=@n92)
  n94 = ScalarParameter() {defaultValue=-1, group=Wrap_Defect_Noise, parameterName=Defect_Bar_Tilerate}
  n95 = NamedRerouteUsage()
  n96 = ComponentMask(None=@n95) {g=true, r=true}
  n97 = Time()
  n98 = ComponentMask(None=VectorParameter() {group=Wrap_Defect_Noise, parameterName=Defect_Bar_Speed, defaultValue=<rgba 0 -0.25 0 1>}.RGB) {g=true, r=true}
  n99 = NamedRerouteUsage()
  n100 = TextureSample(UVs=Add(A=Multiply(A=@n94, B=@n96), B=Multiply(A=@n97, B=@n98)), Tex=@n99) {samplerType=SAMPLERTYPE_LinearColor}
  n101 = ComponentMask(None=@n95) {b=true}
  n102 = Multiply(
   A=TextureSample(UVs=Saturate(None=@n96), Tex=@n99) {samplerType=SAMPLERTYPE_Masks, texture=/Game/Weapons/WeaponSkins/Wraps/Textures/Defect/T_Wrap_Defect_MskB}.A
   B=@n101
  )
  n103 = VectorParameter() {group=Wrap_Defect_Noise, parameterName=Defect_BarB_Speed, defaultValue=<rgba 0 -0.25 0 1>}
  n104 = Add(A=Multiply(B=@n96) {constA=2, constB=2}, B=Multiply(A=Time(), B=ComponentMask(None=@n103.RGB) {g=true, r=true}))
  n105 = Multiply(A=ScalarParameter() {defaultValue=3, group=Wrap_Defect_Noise, parameterName=Defect_Noise_Tilerate}, B=@n96)
  n106 = Multiply(
   A=Time()
   B=ComponentMask(None=VectorParameter() {group=Wrap_Defect_Noise, parameterName=Defect_Noise_Speed, defaultValue=<rgba 1.3 5 0 1>}.RGB) {g=true, r=true}
  )
  n107 = LinearInterpolate(
   A=ScalarParameter() {defaultValue=0.75, group=Wrap_Defect_Noise, parameterName=Defect_Noise_Dark}
   B=ScalarParameter() {defaultValue=1, group=Wrap_Defect_Noise, parameterName=Defect_Noise_Bright}
   Alpha=TextureSample(UVs=Add(A=@n105, B=@n106), Tex=@n99) {samplerType=SAMPLERTYPE_LinearColor}.R
  ) {constA=0.5}
  n108 = Multiply(A=Add(A=TextureSample(UVs=@n104, Tex=@n99) {samplerType=SAMPLERTYPE_LinearColor}.B, B=Multiply(A=@n107, B=@n101)), B=@n101)
  n109 = ScalarParameter() {defaultValue=0.2, group=Wrap_Defect_Layers, parameterName=Defect_Layers_Opacty_5}
  n110 = TextureSample(UVs=@n25, Tex=NamedRerouteUsage())
  n111 = OneMinus(None=@n110.A)
  n112 = Multiply(
   A=Multiply(A=@n76, B=Multiply(A=@n110.R, B=@n111))
   B=TextureSample(UVs=@n24, Tex=NamedRerouteUsage()) {samplerType=SAMPLERTYPE_Masks, texture=&T_Wrap_Defect_MskB}.A
  )
  n113 = TextureSample(UVs=@n20, Tex=NamedRerouteUsage())
  n114 = Multiply(A=@n73, B=Multiply(A=@n113.B, B=OneMinus(None=@n113.A)))
  n115 = TextureSample(UVs=@n19, Tex=NamedRerouteUsage()) {samplerType=SAMPLERTYPE_Masks, texture=&T_Wrap_Defect_MskB}
  n116 = TextureSample(UVs=@n15, Tex=NamedRerouteUsage())
  n117 = Multiply(A=@n116.G, B=OneMinus(None=@n116.A))
  n118 = NamedRerouteUsage()
  n119 = LinearInterpolate(
   A=LinearInterpolate(
    A=LinearInterpolate(B=@n109, Alpha=@n112)
    B=ScalarParameter() {defaultValue=0.5, group=Wrap_Defect_Layers, parameterName=Defect_Layers_Opacty_4}
    Alpha=Multiply(A=@n114, B=@n115.A)
   )
   B=ScalarParameter() {defaultValue=0.6, group=Wrap_Defect_Layers, parameterName=Defect_Layers_Opacty_3}
   Alpha=Multiply(
    A=Multiply(A=@n79, B=@n117)
    B=TextureSample(UVs=@n14, Tex=@n118) {samplerType=SAMPLERTYPE_Masks, texture=&T_Wrap_Defect_MskB}.A
   )
  )
  n120 = ScalarParameter() {defaultValue=0.75, group=Wrap_Defect_Layers, parameterName=Defect_Layers_Opacty_2}
  n121 = TextureSample(UVs=@n10, Tex=NamedRerouteUsage())
  n122 = Multiply(
   A=Multiply(A=@n68, B=Multiply(A=@n121.G, B=@n121.G))
   B=TextureSample(UVs=@n9, Tex=NamedRerouteUsage()) {samplerType=SAMPLERTYPE_Masks, texture=&T_Wrap_Defect_MskB}.A
  )
  n123 = TextureSample(UVs=@n57, Tex=NamedRerouteUsage()) {samplerType=SAMPLERTYPE_Masks, texture=&T_Wrap_Defect_MskB}
  n124 = TextureSample(UVs=@n58, Tex=NamedRerouteUsage())
  n125 = Multiply(A=@n64, B=Multiply(A=@n124.B, B=@n124.B))
  n126 = TextureSample(UVs=@n5, Tex=NamedRerouteUsage())
  n127 = Multiply(A=@n126.R, B=@n126.R)
  n128 = TextureSample(UVs=@n4, Tex=NamedRerouteUsage()) {samplerType=SAMPLERTYPE_Masks, texture=&T_Wrap_Defect_MskB}
  n129 = LinearInterpolate(
   A=LinearInterpolate(
    A=LinearInterpolate(A=@n119, B=@n120, Alpha=@n122)
    B=ScalarParameter() {defaultValue=0.85, group=Wrap_Defect_Layers, parameterName=Defect_Layers_Opacty_1}
    Alpha=Multiply(A=@n123.A, B=@n125)
   )
   B=@n1
   Alpha=Multiply(A=Multiply(A=@n127, B=@n128.A), B=LinearInterpolate(A=@n2.Single_plane_Masking, Alpha=@n3))
  )
  n130 = Multiply(A=@n129, B=VectorParameter() {group=Wrap_Defect_Layers, parameterName=Defect_Color_Layers, defaultValue=<rgba 1 0.4205 0 1>}.RGB)
  n131 = Multiply(A=@n108, B=@n129)
  n132 = Multiply(
   A=Multiply(A=OneMinus(None=@n92), B=OneMinus(None=@n131))
   B=VectorParameter() {group=Wrap_Defect_Layers, parameterName=Defect_Color_Bck, defaultValue=<rgba 0.01 0.0021 0 1>}.RGB
  )
  n133 = Add(
   A=Multiply(
    A=Multiply(A=@n100.G, B=@n102)
    B=VectorParameter() {group=Wrap_Defect_Layers, parameterName=Defect_Color_Bars, defaultValue=<rgba 2 0.8409 0 1>}.RGB
   )
   B=Add(A=Multiply(A=@n108, B=@n130), B=Saturate(None=@n132))
  )
  n134 = SetMaterialAttributes(
   MaterialAttributes=NamedRerouteUsage()
   BaseColor=VectorParameter() {group=Wrap_Defect, parameterName=Defect_BC, defaultValue=<rgba 0 0 0 1>}.RGB
   Roughness=ScalarParameter() {defaultValue=0.32, group=Wrap_Defect, parameterName=Defect_Roughness}
   Specular=ScalarParameter() {defaultValue=0.5, group=Wrap_Defect, parameterName=Defect_Spec}
   Metallic=ScalarParameter() {group=Wrap_Defect, parameterName=Defect_Metalic}
   EmissiveColor=Add(A=Multiply(A=@n89, B=@n90.RGB), B=Add(A=@n93, B=@n133))
  ) {attributeSetTypes [
    #2
    #3
    #5
    #4
    #1
    #6
   ]}
  NamedRerouteDeclaration(None=@n134) {name=Mat_Base, nodeColor=<rgba 0 0.3906 1 1>}
  n135 = Constant4Vector() {constant=<rgba 0 0 0 1>}
  n136 = MaterialFunctionCall(Float4=NamedRerouteDeclaration(None=FunctionInput(Preview=@n135.RGBA) {bUsePreviewValueAsDefault=true, inputName=IsState, inputType=FunctionInput_Vector4, sortPriority=4}) {name=IsState, nodeColor=<rgba 0.1497 1 0.1633 1>}) {materialFunction=&BreakOutFloat4Components}
  n137 = MaterialFunctionCall(Float2=NamedRerouteUsage()) {materialFunction=&BreakOutFloat2Components}
  NamedRerouteDeclaration(None=Saturate(None=Subtract(A=@n136.A, B=@n137.R))) {name=is_character, nodeColor=<rgba 1 0.4219 0 1>}
  n138 = ScalarParameter() {defaultValue=1, group=Wrap_Defect, parameterName=Defect_Fresnel_Exp_Vehicle, sortPriority=100}
  n139 = ScalarParameter() {defaultValue=1, group=Wrap_Defect, parameterName=Defect_Fresnel_Exp_Vehicle_Side, sortPriority=100}
  n140 = MaterialFunctionCall(Float4=NamedRerouteUsage()) {materialFunction=&BreakOutFloat4Components}
  n141 = LinearInterpolate(
   A=LinearInterpolate(
    A=ScalarParameter() {defaultValue=2.5, group=Wrap_Defect, parameterName=Defect_Fresnel_Exp_Weapon, sortPriority=100}
    B=LinearInterpolate(A=@n138, B=@n139, Alpha=@n140.G)
    Alpha=@n140.R
   )
   B=LinearInterpolate(
    A=ScalarParameter() {defaultValue=1, group=Wrap_Defect, parameterName=Defect_Fresnel_Exp_Mech, sortPriority=100}
    B=ScalarParameter() {defaultValue=1, group=Wrap_Defect, parameterName=Defect_Fresnel_Exp_Character, sortPriority=100}
    Alpha=@n140.A
   )
   Alpha=@n140.B
  )
  n142 = ScalarParameter() {defaultValue=1, group=Wrap_Defect, parameterName=Defect_Fresnel_Exp_Backpack, sortPriority=100}
  n143 = MaterialFunctionCall(Float2=NamedRerouteUsage()) {materialFunction=&BreakOutFloat2Components}
  NamedRerouteDeclaration(None=LinearInterpolate(
    A=LinearInterpolate(A=@n141, B=@n142, Alpha=@n143.R)
    B=ScalarParameter() {defaultValue=1, group=Wrap_Defect, parameterName=Defect_Fresnel_Exp_Pickaxe, sortPriority=100}
    Alpha=@n143.G
   )) {name=Fresnel, nodeColor=<rgba 1 0 0.2344 1>}
  n144 = MaterialFunctionCall(
   Object-SafeZone-Virtual-Pivot=NamedRerouteUsage()
   Object-SafeZone-Virtual-CH-Switch=NamedRerouteUsage()
   SinglePlane Normal Axis=NamedRerouteUsage()
   SinglePlane X Axis=NamedRerouteUsage()
  ) {materialFunction=&MF_SafeZone_VirtualPlane}
  n145 = NamedRerouteUsage()
  n146 = LinearInterpolate(A=@n144.Single_plane_UVs, B=@n144.Single_plane_UVs, Alpha=@n145)
  n147 = LinearInterpolate(A=@n144.Single_plane_Masking, Alpha=@n145)
  NamedRerouteDeclaration(None=AppendVector(A=@n146, B=@n147)) {name=UV_Single_Plane, nodeColor=<rgba 0.0312 1 0 1>}
  NamedRerouteDeclaration(None=TextureObjectParameter() {group=Wrap_Defect_Tex, parameterName=Defect__Tex_MskC, texture=&T_BlacksRGB}) {name=Tex_MskC, nodeColor=<rgba 0.8984 1 0 1>}
  NamedRerouteDeclaration(None=Multiply(A=Time(), B=ScalarParameter() {defaultValue=0.5, group=Wrap_Defect_Sine_BumpOffset, parameterName=Defect_Sine_Speed})) {name=Time, nodeColor=<rgba 1 0.9609 0 1>}
  n148 = NamedRerouteUsage()
  n149 = MaterialFunctionCall(Float4=NamedRerouteUsage()) {materialFunction=&BreakOutFloat4Components}
  n150 = LinearInterpolate(
   A=ScalarParameter() {defaultValue=2.5, group=Wrap_Defect_Uvs, parameterName=Defect_Height_Amount__Weapon, sortPriority=150}
   B=LinearInterpolate(
    A=ScalarParameter() {defaultValue=1, group=Wrap_Defect_Uvs, parameterName=Defect_Height_Amount__Vehicle, sortPriority=150}
    B=ScalarParameter() {defaultValue=1, group=Wrap_Defect_Uvs, parameterName=Defect_Height_Amount__Vehicle_Side, sortPriority=150}
    Alpha=@n149.G
   )
   Alpha=@n149.R
  )
  n151 = LinearInterpolate(
   A=ScalarParameter() {defaultValue=1, group=Wrap_Defect_Uvs, parameterName=Defect_Height_Amount__Mech, sortPriority=150}
   B=ScalarParameter() {defaultValue=1, group=Wrap_Defect_Uvs, parameterName=Defect_Height_Amount__Character, sortPriority=150}
   Alpha=@n149.A
  )
  n152 = MaterialFunctionCall(Float2=NamedRerouteUsage()) {materialFunction=&BreakOutFloat2Components}
  n153 = LinearInterpolate(
   A=LinearInterpolate(
    A=LinearInterpolate(A=@n150, B=@n151, Alpha=@n149.B)
    B=ScalarParameter() {defaultValue=1, group=Wrap_Defect_Uvs, parameterName=Defect_Height_Amount__Backpack, sortPriority=150}
    Alpha=@n152.R
   )
   B=ScalarParameter() {defaultValue=1, group=Wrap_Defect_Uvs, parameterName=Defect_Height_Amount__Pickaxe, sortPriority=150}
   Alpha=@n152.G
  )
  n154 = Frac(None=Multiply(A=@n146, B=ScalarParameter() {defaultValue=1, group=Wrap_Defect_Sine_BumpOffset, parameterName=Defect_Sine_TileRate}))
  n155 = Time()
  n156 = ScalarParameter() {defaultValue=0.5, group=Wrap_Defect_Sine_BumpOffset, parameterName=Defect_Sine_Speed}
  n157 = Sine(None=Add(A=ComponentMask(None=@n154) {r=true}, B=Multiply(A=@n155, B=@n156)))
  NamedRerouteDeclaration(None=Multiply(A=Multiply(A=@n148, B=@n153), B=Multiply(A=@n157, B=@n147))) {name=Height_Amount, nodeColor=<rgba 1 0.7266 0 1>}
  n158 = ScalarParameter() {defaultValue=1, group=Wrap_Defect_Uvs, parameterName=Defect_Layers_Scale_Vehicle, sortPriority=150}
  n159 = ScalarParameter() {defaultValue=1, group=Wrap_Defect_Uvs, parameterName=Defect_Layers_Scale_Vehicle_Side, sortPriority=150}
  n160 = MaterialFunctionCall(Float4=NamedRerouteUsage()) {materialFunction=&BreakOutFloat4Components}
  n161 = LinearInterpolate(
   A=LinearInterpolate(
    A=ScalarParameter() {defaultValue=2.5, group=Wrap_Defect_Uvs, parameterName=Defect_Layers_Scale_Weapon, sortPriority=150}
    B=LinearInterpolate(A=@n158, B=@n159, Alpha=@n160.G)
    Alpha=@n160.R
   )
   B=LinearInterpolate(
    A=ScalarParameter() {defaultValue=1, group=Wrap_Defect_Uvs, parameterName=Defect_Layers_Scale_Mech, sortPriority=150}
    B=ScalarParameter() {defaultValue=1, group=Wrap_Defect_Uvs, parameterName=Defect_Layers_Scale_Character, sortPriority=150}
    Alpha=@n160.A
   )
   Alpha=@n160.B
  )
  n162 = ScalarParameter() {defaultValue=1, group=Wrap_Defect_Uvs, parameterName=Defect_Layers_Scale_Backpack, sortPriority=150}
  n163 = MaterialFunctionCall(Float2=NamedRerouteUsage()) {materialFunction=&BreakOutFloat2Components}
  NamedRerouteDeclaration(None=LinearInterpolate(
    A=LinearInterpolate(A=@n161, B=@n162, Alpha=@n163.R)
    B=ScalarParameter() {defaultValue=1, group=Wrap_Defect_Uvs, parameterName=Defect_Layers_Scale_Pickaxe, sortPriority=150}
    Alpha=@n163.G
   )) {name=tileRate, nodeColor=<rgba 1 0 0.2344 1>}
  NamedRerouteDeclaration(None=FunctionInput(Preview=Constant2Vector()) {bUsePreviewValueAsDefault=true, inputName=IsState_B, inputType=FunctionInput_Vector2, sortPriority=4}) {name=IsState_B, nodeColor=<rgba 0.1497 1 0.1633 1>}
  n164 = PreSkinnedLocalBounds()
  n165 = Subtract(A=VertexInterpolator(VS=PreSkinnedPosition()).PS, B=@n164.Min)
  n166 = MaterialFunctionCall(Color=Subtract(A=@n164.Max, B=@n164.Max)) {materialFunction=&SplitComponents}
  n167 = Max(A=Max(A=@n166.R, B=@n166.R), B=@n166.B)
  NamedRerouteDeclaration(None=FunctionInput(Preview=Divide(A=@n165, B=@n167)) {bUsePreviewValueAsDefault=true, inputName=Calcated_Bounds, sortPriority=11}) {name=Caclcated_Bounds, nodeColor=<rgba 0 0.9297 1 1>}
  NamedRerouteDeclaration(None=TextureObjectParameter() {group=Wrap_Defect_Tex, parameterName=Defect__Tex_MskB, texture=&T_BlacksRGB}) {name=Tex_MskB, nodeColor=<rgba 0.8984 1 0 1>}
  n168 = PreSkinnedNormal()
  NamedRerouteDeclaration(None=FunctionInput(Preview=VertexInterpolator(VS=@n168).PS) {bUsePreviewValueAsDefault=true, inputName=Pre-Skinned Local Normal, sortPriority=10}) {name=PreSkin_Normal, nodeColor=<rgba 0.8281 1 0 1>}
  NamedRerouteDeclaration(None=@n35.Normal) {name=Root_Normal, nodeColor=<rgba 0.4062 0 1 1>}
  OneMinus(None=@n102)
  Round(None=Saturate(None=Power(Base=@n6.R, Exp=@n60) {constExponent=70}))
  ScalarParameter() {defaultValue=5, group=Wrap_Defect, parameterName=Defect_Fresnel_Exp}
  ScalarParameter() {defaultValue=2.5, group=Wrap_Defect_Noise, parameterName=Defect_Layers_Tilerate}
  Sine(None=Add(B=NamedRerouteUsage()))
  Sine(None=Add(A=NamedRerouteUsage()) {constB=0.16})
  Sine(None=Add(A=NamedRerouteUsage()) {constB=0.32})
  Sine(None=Add(A=NamedRerouteUsage()) {constB=0.48})
  Sine(None=Add(A=NamedRerouteUsage()) {constB=0.64})
  Sine(None=Add(A=NamedRerouteUsage()) {constB=0.8})
  Sine(None=Add(A=NamedRerouteUsage()) {constB=0.96})
  n169 = VectorParameter() {group=Wrap_Defect_Uvs, parameterName=Defect_Height_Amount_Vehicle, defaultValue=<rgba 0.5 1 1.5 1>}
  n170 = VectorParameter() {group=Wrap_Defect_Uvs, parameterName=Defect_Height_Amount_VehicleSide, defaultValue=<rgba 0.5 1 1.5 1>}
  n171 = MaterialFunctionCall(Float4=NamedRerouteUsage()) {materialFunction=&BreakOutFloat4Components}
  n172 = LinearInterpolate(
   A=LinearInterpolate(
    A=VectorParameter() {group=Wrap_Defect_Uvs, parameterName=Defect_Height_Amount_Weapon, defaultValue=<rgba -1 0 0 1>}.RGB
    B=LinearInterpolate(A=@n169.RGB, B=@n170.RGB, Alpha=@n171.G)
    Alpha=@n171.R
   )
   B=LinearInterpolate(
    A=VectorParameter() {group=Wrap_Defect_Uvs, parameterName=Defect_Height_Amoun_Mech, defaultValue=<rgba 0.5 1 1.5 1>}.RGB
    B=VectorParameter() {group=Wrap_Defect_Uvs, parameterName=Defect_Height_Amoun_Character, defaultValue=<rgba 0.5 1 1.5 1>}.RGB
    Alpha=@n171.A
   )
   Alpha=@n171.B
  )
  n173 = VectorParameter() {group=Wrap_Defect_Uvs, parameterName=Defect_Height_Amount_BackPack, defaultValue=<rgba 0.5 1 1.5 1>}
  n174 = MaterialFunctionCall(Float2=NamedRerouteUsage()) {materialFunction=&BreakOutFloat2Components}
  Transform(None=LinearInterpolate(
    A=LinearInterpolate(A=@n172, B=@n173.RGB, Alpha=@n174.R)
    B=VectorParameter() {group=Wrap_Defect_Uvs, parameterName=Defect_Height_Amount_PickAxe, defaultValue=<rgba 0 -1 0 1>}.RGB
    Alpha=@n174.G
   )) {transformSourceType=TRANSFORMSOURCE_Local}
  Transform(None=Constant3Vector() {constant=<rgba -1 0 0 0>}) {transformSourceType=TRANSFORMSOURCE_Local}
  Transform(None=Constant3Vector() {constant=<rgba 0 1 -0.25 0>}) {transformSourceType=TRANSFORMSOURCE_Local}
 ]}
