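# Material function for the Disarray cosmetic wrap: an animated glitchy-screen look blended over the incoming material by masks, behind the Wrap_Disarray static switch.
# IsState inputs choose offset, rotation and scale per item type, feeding bounds-based projected UVs for pattern and logo.
# Time-hashed bars and threshold pulses jitter those UVs, with per-channel offsets splitting RGB.
# Texture masks, scanlines and static form an inverted emissive, modulated by crevice mask and time of day.
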MaterialFunction /Game/Weapons/WeaponSkins/Wraps/MaterialFunctions/MF_Wrap_Disarray
graph {nodes [
  n1 = VectorParameter() {group=Wrap_Disarray_UVs, parameterName=Disarray_Scanline_Scale, defaultValue=<rgba 5 9 0 1>}
  n2 = Multiply(A=NamedRerouteUsage(), B=ComponentMask(None=Multiply(A=@n1.RGB, B=@n1.RGB)) {g=true, r=true})
  Add(A=@n2, B=ComponentMask(None=VectorParameter() {parameterName=TexCoordOffset, defaultValue=<rgba -0.5 0 0 0>}.RGB) {g=true, r=true})
  n3 = Add(A=NamedRerouteUsage(), B=NamedRerouteUsage())
  n4 = NamedRerouteUsage()
  n5 = ChannelMaskParameter(None=@n4) {group=Wrap_Disarray_Glitch, maskChannel=Green, parameterName=Disarray_Glitch_Direction, defaultValue=<rgba 0 1 0 0>}
  n6 = ScalarParameter() {defaultValue=64, group=Wrap_Disarray_Glitch, parameterName=Disarray_GlitchLine_Tiling}
  n7 = Round(None=Multiply(A=@n5.R, B=@n6))
  n8 = ComponentMask(None=MaterialFunctionCall(Seed=Time(), Float=@n7, Vec2=@n7) {materialFunction=&MF_Hash}.Vec2) {r=true}
  n9 = ConstantBiasScale(None=Multiply(A=@n8, B=@n8)) {bias=-0.5, scale=2}
  n10 = ComponentMask(None=Multiply(A=NamedRerouteUsage()) {constB=2}) {r=true}
  n11 = Add(A=Time(), B=LinearInterpolate(Alpha=Round(None=@n10)) {constB=0.2})
  n12 = ScalarParameter() {defaultValue=0.1, group=Wrap_Disarray_Glitch, parameterName=Disarray_GlitchFrequency}
  n13 = ConstantBiasScale(None=MaterialFunctionCall(Float=Time()) {materialFunction=&MF_Hash}.Float)
  n14 = Subtract(
   A=Frac(None=Multiply(A=@n11, B=@n12) {constB=0.05})
   B=LinearInterpolate(
    A=ScalarParameter() {defaultValue=0.8, group=Wrap_Disarray_Glitch, parameterName=Disarray_GlitchThreshold_Min}
    B=ScalarParameter() {defaultValue=0.9, group=Wrap_Disarray_Glitch, parameterName=Disarray_GlitchThreshold_Max}
    Alpha=Round(None=@n13)
   )
  )
  n15 = Multiply(
   A=ScalarParameter() {defaultValue=0.01, group=Wrap_Disarray_Glitch, parameterName=Disarray_GlitchDistortion_Amt}
   B=Saturate(None=Ceil(None=@n14))
  )
  n16 = Constant3Vector() {constant=<rgba 1 0 0 0>}
  n17 = Multiply(
   A=NamedRerouteDeclaration(None=Multiply(A=@n9, B=@n15) {constB=0.01}) {name=Glitching_Bars, nodeColor=<rgba 1 0 0.4453 1>}
   B=Normalize(VectorInput=ComponentMask(None=@n16) {g=true, r=true})
  )
  n18 = Add(A=@n3, B=@n17)
  Add(A=@n18, B=NamedRerouteUsage())
  n19 = NamedRerouteUsage()
  n20 = ConstantBiasScale(None=Frac(None=Panner(Coordinate=@n19) {speedY=-1})) {bias=-0.5, scale=2}
  n21 = ChannelMaskParameter(None=Multiply(A=@n20, B=@n20)) {group=Wrap_Disarray_Glitch, maskChannel=Green, parameterName=Disarray_Glitch Direction, defaultValue=<rgba 0 1 0 0>}
  n22 = Add(A=Multiply(A=@n21.R, B=ScalarParameter() {defaultValue=3, group=Wrap_Disarray_Static, parameterName=Disarray_Scanline Strength}) {constB=0.1})
  n23 = ChannelMaskParameter(None=AppendVector(A=Constant() {r=6.2832}, B=Constant())) {group=Wrap_Disarray_Glitch, maskChannel=Green, parameterName=Glitch Direction, defaultValue=<rgba 0 1 0 0>}
  n24 = Rotator(Coordinate=NamedRerouteUsage(), Time=@n23.R)
  n25 = Panner(Coordinate=@n24) {speedX=5, speedY=0.1}
  n26 = NamedRerouteUsage()
  n27 = Multiply(
   A=TextureSample(UVs=Panner(Coordinate=@n24) {speedX=6, speedY=-0.05}, Tex=@n26).G
   B=TextureSample(UVs=Panner(Coordinate=@n24) {speedX=-3, speedY=-0.03}, Tex=@n26).R
  )
  n28 = Add(A=Multiply(A=TextureSample(UVs=@n25, Tex=@n26).B) {constB=0.1}, B=@n27)
  n29 = ScalarParameter() {defaultValue=6, group=Wrap_Disarray_Static, parameterName=Disarray_Static Scale}
  n30 = ChannelMaskParameter(None=NamedRerouteUsage()) {group=Wrap_Disarray_Glitch, maskChannel=Green, parameterName=Disarray_Glitch Direction, defaultValue=<rgba 0 1 0 0>}
  n31 = Multiply(A=@n30.R) {constB=0.75}
  n32 = Multiply(A=Time()) {constB=-0.2}
  n33 = Power(Base=Frac(None=Add(A=@n31, B=@n32))) {constExponent=20}
  n34 = ScalarParameter() {defaultValue=0.2, group=Wrap_Disarray_Static, parameterName=Disarray_Sweep Add}
  n35 = Add(
   A=Add(A=Multiply(A=@n28, B=@n29), B=Multiply(A=@n33, B=@n34) {constB=0.2})
   B=ScalarParameter() {defaultValue=0.2, group=Wrap_Disarray, parameterName=Disarray_Background}
  )
  Add(B=Multiply(
    A=VectorParameter() {group=Wrap_Disarray, parameterName=Disarray_Screen Tone, defaultValue=<rgba 0.125 0.0625 0 1>}.RGB
    B=Multiply(A=@n22, B=@n35) {constA=1}
   ))
  n36 = TextureSample() {samplerType=SAMPLERTYPE_Masks, texture=/Game/Weapons/FORT_Rifles/Textures/T_Scar_CM}
  n37 = FunctionInput(Preview=@n36.R) {bUsePreviewValueAsDefault=true, inputName=Mask Primary, inputType=FunctionInput_Scalar, sortPriority=24, previewValue=<1 0 0 1>}
  n38 = FunctionInput(Preview=@n36.B) {bUsePreviewValueAsDefault=true, inputName=Mask Secondary, inputType=FunctionInput_Scalar, sortPriority=24, previewValue=<1 0 0 1>}
  n39 = BlendMaterialAttributes(A=NamedRerouteUsage(), B=NamedRerouteUsage(), Alpha=Saturate(None=Add(A=@n37, B=@n38))) {vertexAttributeBlendType=UseA}
  BlendMaterialAttributes(A=@n39, Alpha=NamedRerouteDeclaration(None=@n38) {name=Accent_Msk, nodeColor=<rgba 0 0.6484 1 1>}) {vertexAttributeBlendType=UseA}
  n40 = Add(A=NamedRerouteUsage(), B=NamedRerouteUsage())
  ComponentMask(None=@n40) {g=true}
  Constant4Vector()
  Constant()
  n41 = TextureSample() {samplerType=SAMPLERTYPE_Normal, texture=/Game/Weapons/FORT_Rifles/T_SCAR_N}
  n42 = Constant3Vector() {constant=<rgba 1 0 0 0>}
  n43 = GetMaterialAttributes(None=FunctionInput(Preview=SetMaterialAttributes!(MaterialAttributes, Normal=@n41.RGB, BaseColor=@n42) {attributeSetTypes [#1, #2]}) {inputName=Material, inputType=FunctionInput_MaterialAttributes}) {attributeGetTypes [
    #2
    #3
    #1
    #4
   ]}
  FunctionOutput(None=StaticSwitchParameter(True=@n39, False=NamedRerouteDeclaration(None=@n43.MaterialAttributes) {name=Root_Material, nodeColor=<rgba 0.7812 1 0 1>}) {group=Customization_Wrap, parameterName=Wrap_Disarray})
  n44 = NamedRerouteUsage()
  Multiply(A=Subtract(A=@n44) {constB=0.35}) {constB=1.5}
  Multiply()
  n45 = NamedRerouteUsage()
  n46 = Constant() {r=0.25}
  n47 = Normalize(VectorInput=ComponentMask(None=Constant3Vector() {constant=<rgba 0 1 0 0>}) {g=true, r=true})
  n48 = Normalize(VectorInput=ComponentMask(None=Constant3Vector() {constant=<rgba 1 0 0 0>}) {g=true, r=true})
  n49 = Multiply(
   A=Multiply(A=If(A=@n45, B=@n46, "A > B"=@n47, "A < B"=@n48), B=NamedRerouteUsage())
   B=ScalarParameter() {defaultValue=1, group=Wrap_Disarray_Glitch, parameterName=Disarray_Glitching_Logo_Strength}
  )
  n50 = NamedRerouteUsage()
  n51 = TextureSample(UVs=Add(A=@n49, B=@n40), Tex=@n50) {samplerSource=SSM_Clamp_WorldGroupSettings, samplerType=SAMPLERTYPE_Masks}
  Multiply(A=@n51.R)
  n52 = VectorParameter() {group=Wrap_Disarray_UVs, parameterName=Disarray_Screen Bounding Box UV Space, defaultValue=<rgba 0 0 2 2>}
  n53 = Divide(A=Subtract(A=NamedRerouteUsage(), B=AppendVector(A=@n52.R, B=@n52.R)), B=AppendVector(A=@n52.B, B=@n52.B))
  n54 = Saturate(None=Sine(None=Saturate(None=@n53)))
  n55 = Multiply(A=ComponentMask(None=@n54) {g=true}, B=ComponentMask(None=@n54) {r=true})
  Multiply(A=@n55, B=@n22)
  n56 = Multiply(A=@n27, B=ScalarParameter() {defaultValue=0.07, group=Wrap_Disarray_Static, parameterName=Disarray_Static UV Distort})
  n57 = Constant()
  n58 = Multiply(A=@n33, B=ScalarParameter() {defaultValue=0.005, group=Wrap_Disarray_Static, parameterName=Disarray_Sweep UV Distort}) {constB=0.02}
  n59 = NamedRerouteUsage()
  NamedRerouteDeclaration(None=Add(A=AppendVector(A=@n56, B=@n57), B=Multiply(A=@n58, B=@n59))) {name=Distorted UVs, nodeColor=<rgba 0 0.3203 1 1>}
  NamedRerouteDeclaration(None=@n18) {name=Glitch Coords, nodeColor=<rgba 0.2656 1 0 1>}
  NamedRerouteDeclaration(None=MaterialFunctionCall!(
    IsStatsB=NamedRerouteUsage()
    Rotation_Axis=Constant3Vector() {constant=<rgba 0 0 0.25 0>}
    Alternative Axies Opacity=Constant()
    IsState=NamedRerouteUsage()
    Caculated_Bounds=NamedRerouteUsage()
    Pre_Skinned_Normals=NamedRerouteUsage()
    Blend_Contrast=ScalarParameter() {defaultValue=20}
   ) {materialFunction=&/MF_TopsSideProjection_UVs}.Combined) {name=UVs_TopSide, nodeColor=<rgba 0 1 0.2734 1>}
  n60 = MaterialFunctionCall(Float4=NamedRerouteUsage()) {materialFunction=&BreakOutFloat4Components}
  NamedRerouteDeclaration(None=MaterialFunctionCall!(
    For_Safe_Zones
    Normal_Map
    Caculated_Bounds=NamedRerouteUsage()
    Pre_Skinned_Normals=NamedRerouteUsage()
    Is Vehicle=@n60.R
    Is VehicleSide=@n60.R
    Is Mech=@n60.R
    Vehicle_Rotation
    UV_Rotation Bypass=NamedRerouteUsage()
    Mech_Rotation
    Uvs_Offset=NamedRerouteUsage()
    Offset-after rotation
    Uvs_TileRate=NamedRerouteUsage()
   ) {materialFunction=&/MF_Projection_Side_Normal}.UVs) {name=UVs_Side_projection, nodeColor=<rgba 0 1 0.2734 1>}
  NamedRerouteDeclaration(None=TextureObjectParameter() {group=Wrap_Disarray, parameterName=Disarray_Tex_Msk, texture=&T_BlacksRGB}) {name=Tex_Msk, nodeColor=<rgba 1 0 0.1172 1>}
  NamedRerouteDeclaration(None=TextureObjectParameter() {group=Wrap_Disarray, parameterName=Disarray_Tex_Msk1, texture=&T_BlacksRGB}) {name=Tex_Msk_1, nodeColor=<rgba 1 0 0.1172 1>}
  NamedRerouteDeclaration(None=TextureObjectParameter() {group=Wrap_Disarray, parameterName=Disarray_Tex_Msk_Scan, texture=&T_BlacksRGB}) {name=Tex_Msk_Scan, nodeColor=<rgba 1 0 0.1172 1>}
  n61 = Multiply(A=Multiply(A=@n4, B=@n6), B=ScalarParameter() {defaultValue=2, group=Wrap_Disarray_Static, parameterName=Disarray_ScaneLine_Tiling})
  n62 = Multiply(A=ScalarParameter() {defaultValue=1, group=Wrap_Disarray_Static, parameterName=Disarray_ScaneLine_Paning_Speed}, B=Time())
  n63 = ConstantBiasScale(None=Frac(None=Add(A=@n61, B=@n62))) {bias=-0.5, scale=2}
  n64 = ScalarParameter() {defaultValue=3, group=Wrap_Disarray_Static, parameterName=Disarray_Scanline Strength}
  n65 = Multiply(A=ComponentMask(None=Multiply(A=@n63, B=@n63)) {g=true}, B=@n64) {constB=0.1}
  NamedRerouteDeclaration(None=Saturate(None=Power(Base=@n65, Exp=@n64))) {name=Scan_Lines, nodeColor=<rgba 0 0.3437 1 1>}
  n66 = MaterialFunctionCall(Float4=NamedRerouteUsage()) {materialFunction=&BreakOutFloat4Components}
  n67 = LinearInterpolate(
   A=VectorParameter() {group=Wrap_Disarray_UVs, parameterName=Disarray_Logo_Offset_Weapon}.RGB
   B=LinearInterpolate(
    A=VectorParameter() {group=Wrap_Disarray_UVs, parameterName=Disarray_Logo_Offset_Vehicle, defaultValue=<rgba 0 0 0 1>}.RGB
    B=VectorParameter() {group=Wrap_Disarray_UVs, parameterName=Disarray_Logo_Offset_VehicleSide, defaultValue=<rgba 0 0 0 1>}.RGB
    Alpha=@n66.G
   )
   Alpha=@n66.R
  )
  n68 = LinearInterpolate(
   A=VectorParameter() {group=Wrap_Disarray_UVs, parameterName=Disarray_Logo_Offset_Mech, defaultValue=<rgba 0 0 0 1>}.RGB
   B=VectorParameter() {group=Wrap_Disarray_UVs, parameterName=Disarray_Logo_Offset_Characte, defaultValue=<rgba 0 0 0 1>}.RGB
   Alpha=@n66.A
  )
  n69 = NamedRerouteUsage()
  NamedRerouteDeclaration(None=LinearInterpolate(
    A=LinearInterpolate(A=@n67, B=@n68, Alpha=@n66.B)
    B=VectorParameter() {group=Wrap_Disarray_UVs, parameterName=Disarray_Logo_Offset_Pickaxe, defaultValue=<rgba 0 0 0 1>}.RGB
    Alpha=MaterialFunctionCall(Float2=@n69) {materialFunction=&BreakOutFloat2Components}.G
   )) {name=Uvs_Logo_offset, nodeColor=<rgba 0 0.625 1 1>}
  n70 = LinearInterpolate(
   A=ScalarParameter() {defaultValue=-0.35, group=Wrap_Disarray, parameterName=Disarray_Crevise_Low}
   B=ScalarParameter() {defaultValue=1, group=Wrap_Disarray, parameterName=Disarray_Crevise_High}
   Alpha=@n44
  )
  n71 = ScalarParameter() {defaultValue=0.25, group=Wrap_Disarray, parameterName=Disarray_Scanline opacity}
  n72 = NamedRerouteUsage()
  n73 = NamedRerouteUsage()
  n74 = TextureSample(UVs=Add(A=Add(A=@n73, B=@n40), B=@n49), Tex=@n50) {samplerSource=SSM_Clamp_WorldGroupSettings, samplerType=SAMPLERTYPE_Masks}
  n75 = Add(A=Add(A=NamedRerouteUsage(), B=@n40), B=@n49)
  n76 = NamedRerouteUsage()
  n77 = TextureSample(UVs=@n18, Tex=@n76) {samplerType=SAMPLERTYPE_Masks}
  n78 = Normalize(VectorInput=ComponentMask(None=Constant3Vector() {constant=<rgba -1 0.5 0 0>}) {g=true, r=true})
  n79 = NamedRerouteUsage()
  n80 = Round(None=ComponentMask(None=Multiply(A=@n79) {constB=2}) {r=true})
  n81 = Multiply(
   A=Add(A=Time(), B=LinearInterpolate(Alpha=@n80) {constB=0.2})
   B=ScalarParameter() {defaultValue=0.2, group=Wrap_Disarray_Glitch, parameterName=Disarray_GlitchFrequency02}
  ) {constB=0.05}
  n82 = ScalarParameter() {defaultValue=0.8, group=Wrap_Disarray_Glitch, parameterName=Disarray_GlitchThreshold02_Min}
  n83 = ScalarParameter() {defaultValue=0.9, group=Wrap_Disarray_Glitch, parameterName=Disarray_GlitchThreshold02_Max}
  n84 = Time()
  n85 = Round(None=ConstantBiasScale(None=MaterialFunctionCall(Float=@n84) {materialFunction=&MF_Hash}.Float))
  n86 = Ceil(None=Subtract(A=Frac(None=@n81), B=LinearInterpolate(A=@n82, B=@n83, Alpha=@n85)))
  n87 = Add(
   A=@n15
   B=Multiply(A=ScalarParameter() {defaultValue=0.01, group=Wrap_Disarray_Glitch, parameterName=Disarray_GlitchDistortion02_Amt}, B=Saturate(None=@n86))
  )
  n88 = ScalarParameter() {defaultValue=0.3, group=Wrap_Disarray_Glitch, parameterName=Disarray_GlitchDistortion_Bck}
  n89 = Multiply(A=NamedRerouteDeclaration(None=Multiply(A=@n87, B=@n78)) {name=Glitching_R, nodeColor=<rgba 1 0 0.4453 1>}, B=@n88)
  n90 = TextureSample(UVs=Add(A=@n17, B=Add(A=@n3, B=@n89)), Tex=@n76) {samplerType=SAMPLERTYPE_Masks}
  n91 = ComponentMask(None=Constant3Vector() {constant=<rgba 0.5 1 0 0>}) {g=true, r=true}
  n92 = NamedRerouteDeclaration(None=Multiply(A=@n87, B=Normalize(VectorInput=@n91))) {name=Glitching_G, nodeColor=<rgba 1 0 0.4453 1>}
  n93 = Add(A=@n17, B=Add(A=@n3, B=Multiply(A=@n92, B=@n88)))
  n94 = Add(
   A=AppendVector(
    A=AppendVector(A=@n51.R, B=@n74.R)
    B=TextureSample(UVs=@n75, Tex=@n50) {samplerSource=SSM_Clamp_WorldGroupSettings, samplerType=SAMPLERTYPE_Masks}.R
   )
   B=AppendVector(A=AppendVector(A=@n77.A, B=@n90.A), B=TextureSample(UVs=@n93, Tex=@n76) {samplerType=SAMPLERTYPE_Masks}.A)
  )
  n95 = Add(
   A=LinearInterpolate(A=@n71, B=Multiply(A=@n72, B=@n71) {constB=0.25}, Alpha=NamedRerouteUsage()) {constA=1}
   B=Multiply(A=Saturate(None=@n94), B=ScalarParameter() {defaultValue=1, group=Wrap_Disarray, parameterName=Disarray_Brightness}) {constB=0.1}
  )
  n96 = NamedRerouteUsage()
  n97 = Multiply(A=Saturate(None=OneMinus(None=@n95)), B=Saturate(None=Add(A=@n96) {constB=0.05}))
  n98 = Constant() {r=0.5}
  n99 = Constant() {r=1}
  n100 = Constant() {r=0.5}
  n101 = Constant() {r=0.25}
  n102 = SetMaterialAttributes!(
   MaterialAttributes=NamedRerouteUsage()
   BaseColor=VectorParameter() {group=Wrap_Disarray, parameterName=Disarray_BaseColor, defaultValue=<rgba 0 0 0 1>}.RGB
   Roughness=ScalarParameter() {group=Wrap_Disarray, parameterName=Disarray_Roughness}
   Specular=ScalarParameter() {group=Wrap_Disarray, parameterName=Disarray_Specular}
   Metallic=ScalarParameter() {group=Wrap_Disarray, parameterName=Disarray_Metalic}
   Normal=NamedRerouteUsage()
   EmissiveColor=Multiply(
    A=Multiply(A=@n70, B=@n97)
    B=MaterialFunctionCall(Value_Morning=@n98, Value_Day=@n99, Value_Evening=@n100, Value_Night=@n101) {materialFunction=&Scalar_Time_of_Day}.Result
   )
  ) {attributeSetTypes [
    #2
    #3
    #5
    #4
    #1
    #6
   ]}
  NamedRerouteDeclaration(None=@n102) {name=Material_Back, nodeColor=<rgba 0 0.3906 1 1>}
  n103 = ScalarParameter() {group=Wrap_Disarray_UVs, parameterName=Disarray_Logo_Rotation_Weapon}
  n104 = MaterialFunctionCall(Float4=NamedRerouteUsage()) {materialFunction=&BreakOutFloat4Components}
  n105 = LinearInterpolate(
   A=ScalarParameter() {group=Wrap_Disarray_UVs, parameterName=Disarray_Logo_Rotation_Vehicle}
   B=ScalarParameter() {group=Wrap_Disarray_UVs, parameterName=Disarray_Logo_Rotation_VehicleSide}
   Alpha=@n104.G
  )
  n106 = ScalarParameter() {group=Wrap_Disarray_UVs, parameterName=Disarray_Logo_Rotation_Mech}
  n107 = ScalarParameter() {group=Wrap_Disarray_UVs, parameterName=Disarray_Logo_Rotation_Character}
  NamedRerouteDeclaration(None=LinearInterpolate(A=LinearInterpolate(A=@n103, B=@n105, Alpha=@n104.R), B=LinearInterpolate(A=@n106, B=@n107, Alpha=@n104.A), Alpha=@n104.B)) {name=Uvs_Logo_Rotation, nodeColor=<rgba 1 0.7266 0 1>}
  n108 = ScalarParameter() {defaultValue=0.5, group=Wrap_Disarray_UVs, parameterName=Disarray_Logo_Scale_Vehicle}
  n109 = ScalarParameter() {defaultValue=0.5, group=Wrap_Disarray_UVs, parameterName=Disarray_Logo_Scale_Vehicle_Side}
  n110 = MaterialFunctionCall(Float4=NamedRerouteUsage()) {materialFunction=&BreakOutFloat4Components}
  n111 = LinearInterpolate(
   A=LinearInterpolate(
    A=ScalarParameter() {defaultValue=0.5, group=Wrap_Disarray_UVs, parameterName=Disarray_Logo_Scale_Weapon}
    B=LinearInterpolate(A=@n108, B=@n109, Alpha=@n110.G)
    Alpha=@n110.R
   )
   B=LinearInterpolate(
    A=ScalarParameter() {defaultValue=0.5, group=Wrap_Disarray_UVs, parameterName=Disarray_Logo_Scale_Mech}
    B=ScalarParameter() {defaultValue=0.5, group=Wrap_Disarray_UVs, parameterName=Disarray_Logo_Scale_Character}
    Alpha=@n110.A
   )
   Alpha=@n110.B
  )
  n112 = ScalarParameter() {defaultValue=0.5, group=Wrap_Disarray_UVs, parameterName=Disarray_Logo_Backpack}
  n113 = MaterialFunctionCall(Float2=NamedRerouteUsage()) {materialFunction=&BreakOutFloat2Components}
  NamedRerouteDeclaration(None=LinearInterpolate(
    A=LinearInterpolate(A=@n111, B=@n112, Alpha=@n113.R)
    B=ScalarParameter() {defaultValue=0.5, group=Wrap_Disarray_UVs, parameterName=Disarray_Logo_Scale_Pickaxe}
    Alpha=@n113.G
   )) {name=Uvs_Logo_Scale, nodeColor=<rgba 0.4062 0 1 1>}
  n114 = MaterialFunctionCall(Float4=NamedRerouteUsage()) {materialFunction=&BreakOutFloat4Components}
  NamedRerouteDeclaration(None=MaterialFunctionCall(
    For_Safe_Zones=StaticBool() {value=true}
    Caculated_Bounds=NamedRerouteUsage()
    Pre_Skinned_Normals=NamedRerouteUsage()
    Is Vehicle=@n114.R
    Is VehicleSide=@n114.R
    Is Mech=@n114.R
    UV_Rotation Bypass=NamedRerouteUsage()
    Uvs_Offset=NamedRerouteUsage()
    Uvs_TileRate=NamedRerouteUsage()
   ) {materialFunction=&MF_Projection_Side_Normal}.UVs) {name=UVs_logo, nodeColor=<rgba 1 0.0938 0 1>}
  n115 = FunctionInput(Preview=@n36.G) {bUsePreviewValueAsDefault=true, inputName=Mask Crevise, inputType=FunctionInput_Scalar, sortPriority=24, previewValue=<1 0 0 1>}
  NamedRerouteDeclaration(None=@n115) {name=Msk_Crevise, nodeColor=<rgba 0.1719 0 1 1>}
  n116 = MaterialFunctionCall(Float4=NamedRerouteUsage()) {materialFunction=&BreakOutFloat4Components}
  n117 = LinearInterpolate(
   A=VectorParameter() {group=Wrap_Disarray_UVs, parameterName=Disarray_Offset_Weapon}.RGB
   B=LinearInterpolate(
    A=VectorParameter() {group=Wrap_Disarray_UVs, parameterName=Disarray_Offset_Vehicle, defaultValue=<rgba 0 0 0 1>}.RGB
    B=VectorParameter() {group=Wrap_Disarray_UVs, parameterName=Disarray_Offset_VehicleSide, defaultValue=<rgba 0 0 0 1>}.RGB
    Alpha=@n116.G
   )
   Alpha=@n116.R
  )
  n118 = LinearInterpolate(
   A=VectorParameter() {group=Wrap_Disarray_UVs, parameterName=Disarray_Offset_Mech, defaultValue=<rgba 0 0 0 1>}.RGB
   B=VectorParameter() {group=Wrap_Disarray_UVs, parameterName=Disarray_Offset_Characte, defaultValue=<rgba 0 0 0 1>}.RGB
   Alpha=@n116.A
  )
  n119 = NamedRerouteUsage()
  NamedRerouteDeclaration(None=LinearInterpolate(
    A=LinearInterpolate(A=@n117, B=@n118, Alpha=@n116.B)
    B=VectorParameter() {group=Wrap_Disarray_UVs, parameterName=Disarray_Offset_Pickaxe, defaultValue=<rgba 0 0 0 1>}.RGB
    Alpha=MaterialFunctionCall(Float2=@n119) {materialFunction=&BreakOutFloat2Components}.G
   )) {name=Side_offset, nodeColor=<rgba 0 0.625 1 1>}
  n120 = ScalarParameter() {group=Wrap_Disarray_UVs, parameterName=Disarray_Rotation_Weapon}
  n121 = MaterialFunctionCall(Float4=NamedRerouteUsage()) {materialFunction=&BreakOutFloat4Components}
  n122 = LinearInterpolate(
   A=ScalarParameter() {group=Wrap_Disarray_UVs, parameterName=Disarray_Rotation_Vehicle}
   B=ScalarParameter() {group=Wrap_Disarray_UVs, parameterName=Disarray_Rotation_VehicleSide}
   Alpha=@n121.G
  )
  n123 = ScalarParameter() {group=Wrap_Disarray_UVs, parameterName=Disarray_Rotation_Mech}
  n124 = ScalarParameter() {group=Wrap_Disarray_UVs, parameterName=Disarray_Rotation_Character}
  NamedRerouteDeclaration(None=LinearInterpolate(A=LinearInterpolate(A=@n120, B=@n122, Alpha=@n121.R), B=LinearInterpolate(A=@n123, B=@n124, Alpha=@n121.A), Alpha=@n121.B)) {name=Side_Rotation, nodeColor=<rgba 1 0.7266 0 1>}
  NamedRerouteDeclaration(None=@n55) {name=Screan_effect, nodeColor=<rgba 0.4297 1 0 1>}
  NamedRerouteDeclaration(None=@n2) {name=TexCoord, nodeColor=<rgba 1 0 0.4453 1>}
  n125 = MaterialFunctionCall(Float3=Abs(None=NamedRerouteUsage())) {materialFunction=&BreakOutFloat3Components}
  n126 = Constant() {r=0.5}
  n127 = NamedRerouteUsage()
  n128 = MaterialFunctionCall(Float3=@n127) {materialFunction=&BreakOutFloat3Components}
  n129 = AppendVector(A=@n128.G, B=@n128.G)
  n130 = ComponentMask(None=@n127) {desc=YX, g=true, r=true}
  n131 = MaterialFunctionCall(Float4=NamedRerouteUsage()) {materialFunction=&BreakOutFloat4Components}
  n132 = Max(A=Max(A=@n131.R, B=@n131.R), B=@n131.B)
  n133 = ComponentMask(None=@n127) {b=true, desc=YZ, g=true}
  n134 = ComponentMask(None=@n127) {b=true, desc=XZ, r=true}
  NamedRerouteDeclaration(None=If(A=@n125.B, B=@n126, "A > B"=LinearInterpolate(A=@n129, B=@n130, Alpha=@n132), "A < B"=If(A=@n125.R, B=@n126, "A > B"=@n133, "A < B"=@n134))) {name=UVS_Cheap_Tri, nodeColor=<rgba 1 0 0.8438 1>}
  n135 = MaterialFunctionCall(Color=Abs(None=NamedRerouteUsage())) {materialFunction=&SplitComponents}
  n136 = MaterialFunctionCall(Float4=NamedRerouteUsage()) {materialFunction=&BreakOutFloat4Components}
  n137 = LinearInterpolate(A=@n135.R, B=LinearInterpolate(A=@n135.B, B=@n135.B, Alpha=@n136.G), Alpha=@n136.R)
  NamedRerouteDeclaration(None=MaterialFunctionCall(
    In=LinearInterpolate(A=@n137, B=@n135.G, Alpha=@n136.B)
    Contrast=ScalarParameter() {defaultValue=1, group=Wrap_WomanDay01, parameterName=WD01_TriPlanerMsk_Contrast, sortPriority=9}
   ) {materialFunction=&CheapContrast}.Result) {name=Tri_Mask, nodeColor=<rgba 1 0.4688 0 1>}
  NamedRerouteDeclaration(None=FunctionInput(Preview=Constant4Vector().RGBA) {bUsePreviewValueAsDefault=true, inputName=IsState, inputType=FunctionInput_Vector4, sortPriority=4}) {name=IsState, nodeColor=<rgba 0.1497 1 0.1633 1>}
  NamedRerouteDeclaration(None=FunctionInput(Preview=Constant2Vector()) {bUsePreviewValueAsDefault=true, inputName=IsState_B, inputType=FunctionInput_Vector2, sortPriority=4}) {name=IsState_B, nodeColor=<rgba 0.1497 1 0.1633 1>}
  n138 = PreSkinnedLocalBounds()
  n139 = Subtract(A=VertexInterpolator(VS=PreSkinnedPosition()).PS, B=@n138.Min)
  n140 = MaterialFunctionCall(Color=Subtract(A=@n138.Max, B=@n138.Max)) {materialFunction=&SplitComponents}
  n141 = Max(A=Max(A=@n140.R, B=@n140.R), B=@n140.B)
  NamedRerouteDeclaration(None=FunctionInput(Preview=Divide(A=@n139, B=@n141)) {bUsePreviewValueAsDefault=true, inputName=Calcated_Bounds, sortPriority=11}) {name=Caclcated_Bounds, nodeColor=<rgba 0 0.9297 1 1>}
  n142 = ScalarParameter() {defaultValue=1, group=Wrap_Disarray_UVs, parameterName=Disarray_Scale_Vehicle}
  n143 = ScalarParameter() {defaultValue=1, group=Wrap_Disarray_UVs, parameterName=Disarray_Scale_Vehicle_Side}
  n144 = MaterialFunctionCall(Float4=NamedRerouteUsage()) {materialFunction=&BreakOutFloat4Components}
  n145 = LinearInterpolate(
   A=LinearInterpolate(
    A=ScalarParameter() {defaultValue=1, group=Wrap_Disarray_UVs, parameterName=Disarray_Scale_Weapon}
    B=LinearInterpolate(A=@n142, B=@n143, Alpha=@n144.G)
    Alpha=@n144.R
   )
   B=LinearInterpolate(
    A=ScalarParameter() {defaultValue=1, group=Wrap_Disarray_UVs, parameterName=Disarray_Scale_Mech}
    B=ScalarParameter() {defaultValue=1, group=Wrap_Disarray_UVs, parameterName=Disarray_Scale_Character}
    Alpha=@n144.A
   )
   Alpha=@n144.B
  )
  n146 = ScalarParameter() {defaultValue=1, group=Wrap_Disarray_UVs, parameterName=Disarray_Scale_Backpack}
  n147 = MaterialFunctionCall(Float2=NamedRerouteUsage()) {materialFunction=&BreakOutFloat2Components}
  NamedRerouteDeclaration(None=LinearInterpolate(
    A=LinearInterpolate(A=@n145, B=@n146, Alpha=@n147.R)
    B=ScalarParameter() {defaultValue=1, group=Wrap_Disarray_UVs, parameterName=Disarray_Scale_Pickaxe}
    Alpha=@n147.G
   )) {name=Side_Scale, nodeColor=<rgba 0.4062 0 1 1>}
  n148 = PreSkinnedNormal()
  NamedRerouteDeclaration(None=FunctionInput(Preview=VertexInterpolator(VS=@n148).PS) {bUsePreviewValueAsDefault=true, inputName=Pre-Skinned Local Normal, sortPriority=10}) {name=PreSkin_Normal, nodeColor=<rgba 0.8281 1 0 1>}
  NamedRerouteDeclaration(None=@n43.Normal) {name=Root_Normal, nodeColor=<rgba 0.4062 0 1 1>}
  NamedRerouteUsage()
  NamedRerouteUsage()
  Round(None=@n65)
  Saturate(None=@n37)
  Saturate(None=@n77.A)
  n149 = ConstantBiasScale(None=Frac(None=@n19)) {bias=-0.5, scale=2}
  n150 = ComponentMask(None=Multiply(A=@n149, B=@n149)) {g=true}
  n151 = ScalarParameter() {defaultValue=3, group=Wrap_Disarray_Static, parameterName=Disarray_Scanline Strength}
  Saturate(None=Round(None=Multiply(A=@n150, B=@n151) {constB=0.1}))
  TextureCoordinate() {coordinateIndex=1}
  TextureCoordinate() {coordinateIndex=1}
  TextureCoordinate() {coordinateIndex=1}
  VertexNormalWS()
 ]}
