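# Material function for the EclipseS wrap skin on weapons, vehicles and mechs, overriding base color, metallic, specular, roughness and emissive.
# Side projection from pre-skinned bounds and normals supplies per-type UVs for base, mask and two time-scrolled, warped star layers, limited by normal-direction masks.
# A timestamp parameter drives a decaying reactive pulse that shifts emissive color, strength and mask.
# A static switch enables the result, blended over the input material by the customization mask.
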MaterialFunction /Game/Weapons/WeaponSkins/Wraps/MaterialFunctions/MF_Wrap_EclipseS
graph {nodes [
  n1 = Sine(None=Multiply(A=Time()) {constB=0.1}) {period=0.5}
  Clamp(None=Divide(A=Add(A=@n1))) {minDefault=0.3}
  n2 = VectorParameter() {group=Wrap_EclipseS, parameterName=ECS_BaseColor_TintForRoll, sortPriority=-100, defaultValue=<rgba 1 1 1 0>}
  n3 = TextureObjectParameter() {group=Wrap__EclipseS_Textures, parameterName=ECS_BaseColor_01, sortPriority=11, texture=&T_BlacksRGB}
  n4 = Constant()
  n5 = MaterialFunctionCall(
   R=FunctionInput(Preview=Constant()) {inputName=isVehicle, inputType=FunctionInput_Scalar, sortPriority=5}
   G=FunctionInput(Preview=@n4) {bUsePreviewValueAsDefault=true, inputName=Side Project Vehicle, inputType=FunctionInput_Scalar}
   B=FunctionInput(Preview=@n4) {bUsePreviewValueAsDefault=true, inputName=isMech, inputType=FunctionInput_Scalar, sortPriority=4}
  ) {materialFunction=&AppendMany}
  n6 = MaterialFunctionCall(Color=@n5.RGB) {materialFunction=&SplitComponents}
  n7 = VectorParameter() {group=Wrap_EclipseS_UVs, parameterName=ECS_Offset_Weapon, defaultValue=<rgba 0 0 0 1>}
  n8 = MaterialFunctionCall(Color=@n5.RGB) {materialFunction=&SplitComponents}
  n9 = LinearInterpolate(
   A=VectorParameter() {group=Wrap_EclipseS_UVs, parameterName=ECS_Offset_Vehicle, defaultValue=<rgba 0 0 0 1>}.RGB
   B=VectorParameter() {group=Wrap_EclipseS_UVs, parameterName=ECS_Offset_Vehicle_Side, defaultValue=<rgba 0 0 0 1>}.RGB
   Alpha=@n8.G
  )
  n10 = ScalarParameter() {defaultValue=1, group=Wrap_EclipseS_UVs, parameterName=ECS_TileRate_Weapon}
  n11 = LinearInterpolate(
   A=ScalarParameter() {defaultValue=1, group=Wrap_EclipseS_UVs, parameterName=ECS_TileRate_Vehicle}
   B=ScalarParameter() {defaultValue=1, group=Wrap_EclipseS_UVs, parameterName=ECS_TileRate_VehicleSide}
   Alpha=@n8.G
  )
  n12 = PreSkinnedLocalBounds()
  n13 = Subtract(A=VertexInterpolator(VS=PreSkinnedPosition()).PS, B=@n12.Min)
  n14 = Subtract(A=@n12.Max, B=@n12.Max)
  n15 = FunctionInput(Preview=VertexInterpolator(VS=PreSkinnedNormal()).PS) {bUsePreviewValueAsDefault=true, inputName=Pre-Skinned Local Normal, sortPriority=10}
  n16 = MaterialFunctionCall!(
   For_Safe_Zones
   Normal_Map
   Caculated_Bounds=FunctionInput(Preview=Divide(A=@n13, B=@n14)) {bUsePreviewValueAsDefault=true, inputName=Calcated_Bounds, sortPriority=11}
   Pre_Skinned_Normals=@n15
   Is Vehicle=@n6.R
   Is VehicleSide=@n6.R
   Is Mech=@n6.R
   Vehicle_Rotation=ScalarParameter() {group=Wrap_EclipseS_UVs, parameterName=ECS_Rotation_Vehicle}
   UV_Rotation Bypass
   Mech_Rotation=ScalarParameter() {defaultValue=0.25, group=Wrap_EclipseS_UVs, parameterName=ECS_Rotation_Mech -----}
   Uvs_Offset=LinearInterpolate(
    A=LinearInterpolate(A=@n7.RGB, B=@n9, Alpha=@n8.R)
    B=VectorParameter() {group=Wrap_EclipseS_UVs, parameterName=ECS_Offset_Mech -----, defaultValue=<rgba 0 0 0 1>}.RGB
    Alpha=@n8.B
   )
   Offset-after rotation
   Uvs_TileRate=LinearInterpolate(
    A=LinearInterpolate(A=@n10, B=@n11, Alpha=@n8.R)
    B=ScalarParameter() {defaultValue=1, group=Wrap_EclipseS_UVs, parameterName=ECS_TileRate_Mech -----}
    Alpha=@n8.B
   )
  ) {materialFunction=&/MF_Projection_Side_Normal}
  n17 = TextureSample(UVs=@n16.UVs, Tex=@n3)
  n18 = VectorParameter() {group=Wrap_EclipseS, parameterName="ECS_BaseColor_Stars01 - Roughness=A", defaultValue=<rgba 11.6058 1.2211 20 0.25>}
  n19 = TextureObjectParameter() {group=Wrap__EclipseS_Textures, parameterName=ECS_Mask_02, sortPriority=20, texture=&T_BlacksRGB}
  n20 = ChannelMaskParameter(None=TextureSample(UVs=@n16.UVs, Tex=@n19).RGBA) {group=Wrap_EclipseS, maskChannel=Alpha, parameterName=ECS_StarsBlockOut_ChannelMask, sortPriority=50, defaultValue=<rgba 0 0 0 1>}
  n21 = Multiply(A=Time(), B=ScalarParameter() {group=Wrap_EclipseS, parameterName=ECS_Stars01_GradientSpeed, sortPriority=50})
  n22 = Time()
  n23 = VectorParameter() {group=Wrap_EclipseS, parameterName=ECS_Stars01_Speed, sortPriority=50, defaultValue=<rgba 0 0 0 1>}
  n24 = ComponentMask(None=@n23.RGB) {g=true, r=true}
  n25 = TextureSample(UVs=Add(A=Multiply(A=@n22, B=@n24), B=@n16.UVs), Tex=@n19)
  n26 = ChannelMaskParameter(None=@n25.RGBA) {group=Wrap_EclipseS, maskChannel=Blue, parameterName=ECS_Stars_WarpMask, sortPriority=50, defaultValue=<rgba 0 0 1 0>}
  n27 = Add(
   A=Multiply(
    A=Add(A=@n21, B=@n16.UVs)
    B=ScalarParameter() {defaultValue=1, group=Wrap_EclipseS, parameterName=ECS_Stars01_GradientTileRate, sortPriority=50}
   )
   B=@n26.R
  )
  n28 = ChannelMaskParameter(None=@n25.RGBA) {group=Wrap_EclipseS, maskChannel=Green, parameterName=ECS_Stars_ChannelMask, sortPriority=50, defaultValue=<rgba 0 1 0 0>}
  n29 = Multiply(
   A=ChannelMaskParameter(None=TextureSample(UVs=@n27, Tex=@n19).RGBA) {group=Wrap_EclipseS, parameterName=ECS_StarsGradient_ChannelMask, sortPriority=50}.R
   B=@n28.R
  )
  n30 = Multiply(A=@n29, B=Add(A=@n29, B=@n28.R))
  n31 = MaterialFunctionCall(Color=Abs(None=@n15)) {materialFunction=&SplitComponents}
  n32 = LinearInterpolate(A=Add(A=@n31.R, B=@n31.R), B=Add(A=@n31.B, B=@n31.B))
  n33 = Add(A=@n31.R, B=@n31.R)
  n34 = OneMinus(None=MaterialFunctionCall(
    In=LinearInterpolate(A=@n32, B=@n33)
    Contrast=ScalarParameter() {defaultValue=0.1, group=Wrap_EclipseS, parameterName=ECS_Mask_Contrast, sortPriority=150}
   ) {materialFunction=&CheapContrast}.Result)
  n35 = MaterialFunctionCall(Color=Abs(None=@n15)) {materialFunction=&SplitComponents}
  n36 = MaterialFunctionCall(Color=@n5.RGB) {materialFunction=&SplitComponents}
  n37 = LinearInterpolate(A=LinearInterpolate(A=@n35.R, B=LinearInterpolate(A=@n35.B, B=@n35.B, Alpha=@n36.G), Alpha=@n36.R), B=@n35.G, Alpha=@n36.B)
  n38 = ScalarParameter() {defaultValue=5, group=Wrap_EclipseS, parameterName=ECS_Alt+ZMask_Contrast, sortPriority=150}
  n39 = LinearInterpolate(
   A=LinearInterpolate(
    A=Constant()
    B=Saturate(None=@n34)
    Alpha=ScalarParameter() {defaultValue=1, group=Wrap_EclipseS, parameterName=ECS_Mask_AltAxisActive -----, sortPriority=150}
   )
   B=Saturate(None=MaterialFunctionCall(In=@n37, Contrast=@n38) {materialFunction=&CheapContrast}.Result)
   Alpha=ScalarParameter() {defaultValue=1, group=Wrap_EclipseS, parameterName=ECS_Alt+ZMask_Active -----, sortPriority=150}
  )
  n40 = Multiply(A=Multiply(A=@n20.R, B=Saturate(None=@n30)), B=@n39)
  n41 = VectorParameter() {group=Wrap_EclipseS, parameterName="ECS_BaseColor_Stars02 - Roughness=A", defaultValue=<rgba 5.4005 20 3.0454 0.5>}
  n42 = Multiply(A=Time(), B=ScalarParameter() {group=Wrap_EclipseS, parameterName=ECS_Stars02_GradientSpeed, sortPriority=50})
  n43 = Time()
  n44 = VectorParameter() {group=Wrap_EclipseS, parameterName=ECS_Stars02_Speed, sortPriority=50, defaultValue=<rgba 0 0 0 1>}
  n45 = ComponentMask(None=@n44.RGB) {g=true, r=true}
  n46 = TextureSample(UVs=Add(A=Multiply(A=@n43, B=@n45), B=@n16.UVs), Tex=@n19)
  n47 = ChannelMaskParameter(None=@n46.RGBA) {group=Wrap_EclipseS, maskChannel=Blue, parameterName=ECS_Stars_WarpMask, sortPriority=50, defaultValue=<rgba 0 0 1 0>}
  n48 = Add(
   A=Multiply(
    A=Add(A=@n42, B=@n16.UVs)
    B=ScalarParameter() {defaultValue=1, group=Wrap_EclipseS, parameterName=ECS_Stars02_GradientTileRate, sortPriority=50}
   )
   B=@n47.R
  )
  n49 = ChannelMaskParameter(None=@n46.RGBA) {group=Wrap_EclipseS, maskChannel=Green, parameterName=ECS_Stars_ChannelMask, sortPriority=50, defaultValue=<rgba 0 1 0 0>}
  n50 = Multiply(
   A=ChannelMaskParameter(None=TextureSample(UVs=@n48, Tex=@n19).RGBA) {group=Wrap_EclipseS, parameterName=ECS_StarsGradient_ChannelMask, sortPriority=50}.R
   B=@n49.R
  )
  n51 = Multiply(A=@n50, B=Add(A=@n50, B=@n49.R))
  n52 = Multiply(A=Multiply(A=@n20.R, B=Saturate(None=@n51)), B=@n39)
  n53 = LinearInterpolate(A=LinearInterpolate(A=ComponentMask(None=@n17.RGBA) {b=true, g=true, r=true}, B=@n18.RGB, Alpha=@n40), B=@n41.RGB, Alpha=@n52)
  n54 = TextureObjectParameter() {group=Wrap__EclipseS_Textures, parameterName=ECS_Mask_01, sortPriority=20, texture=&T_BlacksRGB}
  n55 = TextureSample(UVs=@n16.UVs, Tex=@n54)
  n56 = ChannelMaskParameter(None=@n55.RGBA) {group=Wrap_EclipseS, maskChannel=Green, parameterName=ECS_Emissive_ChannelMask, sortPriority=0, defaultValue=<rgba 0 1 0 0>}
  n57 = Time()
  n58 = ScalarParameter() {defaultValue=-9999999, group=Wrap_EclipseS_Reactive, parameterName=TimeFrom-BP----------------, sortPriority=-600}
  n59 = Saturate(None=Divide(
    A=Subtract(A=@n57, B=@n58)
    B=ScalarParameter() {defaultValue=1.5, group=Wrap_EclipseS_Reactive, parameterName=ECS_AnimationDuration, sortPriority=33}
   ))
  n60 = ScalarParameter() {defaultValue=1.1, group=Wrap_EclipseS_Reactive, parameterName=ECS_Curve_Amount, sortPriority=34}
  n61 = ScalarParameter() {group=Wrap_EclipseS_Reactive, parameterName=ECS_TestReactive, sortPriority=300}
  n62 = LinearInterpolate(A=OneMinus(None=Power(Base=@n59, Exp=@n60)) {desc=inverts curve from 0-1 to 1-0}, B=@n61, Alpha=@n61)
  n63 = LinearInterpolate(A=@n56.R, B=Add(A=@n52, B=@n56.R), Alpha=@n62)
  n64 = OneMinus(None=@n63)
  n65 = LinearInterpolate(
   A=ScalarParameter() {group=Wrap_EclipseS, parameterName=ECS_Roughness_Min, sortPriority=2}
   B=ScalarParameter() {defaultValue=1, group=Wrap_EclipseS, parameterName=ECS_Roughness_Max, sortPriority=2}
   Alpha=ComponentMask(None=@n17.RGBA) {a=true}
  )
  n66 = LinearInterpolate(
   A=ScalarParameter() {defaultValue=0.5, group=Wrap_EclipseS, parameterName=ECS_Emissive_Amount-----, sortPriority=0}
   B=ScalarParameter() {defaultValue=3, group=Wrap_EclipseS, parameterName=ECS_Emissive_ReactiveAmount, sortPriority=0}
   Alpha=@n62
  )
  n67 = LinearInterpolate(A=Constant() {r=1}, Alpha=@n62)
  n68 = VectorParameter() {group=Wrap_EclipseS, parameterName=ECS_Emissive_Color, sortPriority=0, defaultValue=<rgba 1 0.0885 0.0885 1>}
  n69 = VectorParameter() {group=Wrap_EclipseS, parameterName=ECS_Emissive_Reactive, sortPriority=0, defaultValue=<rgba 0 0.9734 1 1>}
  n70 = LinearInterpolate(A=@n68.RGB, B=@n69.RGB, Alpha=@n62)
  n71 = GetMaterialAttributes(None=FunctionInput() {inputName=Material, inputType=FunctionInput_MaterialAttributes, sortPriority=-20}) {attributeGetTypes [#1]}
  n72 = LinearInterpolate(
   A=ScalarParameter() {group=Wrap_EclipseS, parameterName=ECS_Specular_Min, sortPriority=2}
   B=ScalarParameter() {defaultValue=1, group=Wrap_EclipseS, parameterName=ECS_Specular_Max, sortPriority=2}
   Alpha=ChannelMaskParameter(None=@n55.RGBA) {group=Wrap_EclipseS, parameterName=ECS_Specular_ChannelMask, sortPriority=2}.R
  )
  n73 = ScalarParameter() {defaultValue=0.2, group=Wrap_EclipseS, parameterName=ECS_Stars01_Specular, sortPriority=50}
  n74 = SetMaterialAttributes!(
   MaterialAttributes=@n71.MaterialAttributes
   BaseColor=Multiply(A=@n2.RGB, B=LinearInterpolate(A=@n53, B=@n64, Alpha=@n63))
   Metallic=ScalarParameter() {group=Wrap_EclipseS, parameterName=ECS_Metal}
   Specular=LinearInterpolate(
    A=LinearInterpolate(A=@n72, B=@n73, Alpha=@n40)
    B=ScalarParameter() {defaultValue=0.5, group=Wrap_EclipseS, parameterName=ECS_Stars02_Specular, sortPriority=50}
    Alpha=@n52
   )
   Roughness=LinearInterpolate(A=LinearInterpolate(A=@n65, B=@n18.A, Alpha=@n40), B=@n41.A, Alpha=@n52)
   EmissiveColor=Multiply(A=Multiply(A=@n66, B=@n67), B=Multiply(A=@n70, B=@n63))
  ) {attributeSetTypes [
    #2
    #3
    #4
    #5
    #6
   ]}
  n75 = FunctionInput() {
   bUsePreviewValueAsDefault=true
   desc=base customization
   inputName="Customization Mask (_CM B Top left)"
   inputType=FunctionInput_Scalar
   sortPriority=49
  }
  FunctionOutput(None=StaticSwitchParameter(True=BlendMaterialAttributes(A=@n71.MaterialAttributes, B=@n74, Alpha=@n75) {vertexAttributeBlendType=UseA}, False=@n71.MaterialAttributes) {group=Customization_Wrap, parameterName=Wrap_EclipseS})
  n76 = Clamp(None=@n16.Raw_Normal)
  LinearInterpolate(
   A=@n71.Normal
   B=LinearInterpolate(
    A=@n71.Normal
    B=Add(A=@n71.Normal, B=@n76)
    Alpha=ScalarParameter() {defaultValue=1, group=Wrap_EclipseS, parameterName=ECS_Normal_Intensity-----, sortPriority=1}
   )
   Alpha=@n39
  )
 ]}
