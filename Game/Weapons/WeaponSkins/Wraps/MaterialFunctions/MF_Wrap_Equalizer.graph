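# Wrap material function that overlays an animated, audio-reactive equalizer on wrapped objects, gated by the Equalizer static switch and a customization mask.
# MF_WrapBounds coordinates, with vehicle and mech variants, are blended by pre-skinned normal to project grid cells.
# Custom HLSL draws five pixelated bands from timed pulses maxed with emote and event audio, which also shifts colors.
# Noise, grid lines and cubemap finish the look; Wrap FFT nodes are unconnected.
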MaterialFunction /Game/Weapons/WeaponSkins/Wraps/MaterialFunctions/MF_Wrap_Equalizer
graph {nodes [
  n1 = FunctionInput() {inputName=Material, inputType=FunctionInput_MaterialAttributes}
  n2 = GetMaterialAttributes(None=@n1) {attributeGetTypes [#1, #2]}
  n3 = FunctionInput(Preview=Constant() {r=1}) {bUsePreviewValueAsDefault=true, inputName=isVehicle, inputType=FunctionInput_Scalar, sortPriority=2}
  n4 = PreSkinnedPosition()
  n5 = MaterialFunctionCall!(
   is Vehicle=@n3
   Object Bounds Max Adjustment=FunctionInput() {bUsePreviewValueAsDefault=true, inputName=Object Bounds Max Adjustment, sortPriority=3}
   Object Bounds Min Adjustment=FunctionInput() {bUsePreviewValueAsDefault=true, inputName=Object Bounds Min Adjustment, sortPriority=4}
   Object Bounds Center Adjustment=FunctionInput() {bUsePreviewValueAsDefault=true, inputName=Object Bounds Center Adjustment, sortPriority=5}
   isMech
   Side Project Vehicle
   Local Position=FunctionInput(Preview=VertexInterpolator(VS=@n4).PS) {bUsePreviewValueAsDefault=true, inputName=Local Position, sortPriority=1}
  ) {materialFunction=&/MF_WrapBounds}
  n6 = MaterialFunctionCall(Color=@n5.Result) {materialFunction=&SplitComponents}
  n7 = AppendVector(A=@n6.B, B=@n6.B)
  n8 = MaterialFunctionCall(Color=@n5.Result) {materialFunction=&SplitComponents}
  n9 = AppendVector(A=@n8.B, B=@n8.B)
  n10 = AppendVector(A=@n8.R, B=OneMinus(None=@n8.B))
  n11 = FunctionInput() {bUsePreviewValueAsDefault=true, inputName=isMech, inputType=FunctionInput_Scalar, sortPriority=2}
  n12 = Frac(None=LinearInterpolate(A=LinearInterpolate(A=@n7, B=@n9, Alpha=@n3), B=@n10, Alpha=@n11))
  n13 = MaterialFunctionCall(
   X=ScalarParameter() {defaultValue=100, group=Equalizer, parameterName=CellsX}
   Y=Divide(A=ScalarParameter() {defaultValue=200, group=Equalizer, parameterName=CellsY})
  ) {materialFunction=&MakeFloat2}
  n14 = TextureObject() {samplerType=SAMPLERTYPE_Masks, texture=/Game/Characters/Player/Male/Medium/Bodies/M_MED_DJ_Remix/Textures/TechArt_Textures/T_GridCell}
  n15 = TextureSample(UVs=Multiply(A=@n12, B=@n13.Result), Tex=@n14) {samplerType=SAMPLERTYPE_Masks, texture=&T_GridCell}
  n16 = AppendVector(A=@n6.R, B=@n6.R)
  n17 = AppendVector(A=@n8.G, B=@n8.G)
  n18 = Frac(None=LinearInterpolate(A=LinearInterpolate(A=@n16, B=@n17, Alpha=@n3), B=@n10, Alpha=@n11))
  n19 = Multiply(A=@n18, B=@n13.Result)
  n20 = TextureSample(UVs=@n19, Tex=@n14) {samplerType=SAMPLERTYPE_Masks, texture=&T_GridCell}
  n21 = FunctionInput(Preview=VertexInterpolator(VS=PreSkinnedNormal()).PS) {bUsePreviewValueAsDefault=true, inputName=Pre-Skinned Local Normal, sortPriority=10}
  n22 = ScalarParameter() {defaultValue=1, group=Equalizer, parameterName=Equ_Tri_Mask_Contrast}
  n23 = MaterialFunctionCall(In=Abs(None=ComponentMask(None=@n21) {b=true}), Contrast=@n22) {materialFunction=&CheapContrast}
  n24 = LinearInterpolate(A=LinearInterpolate(B=AppendVector(A=@n8.G, B=@n8.G), Alpha=@n3), B=@n10, Alpha=@n11)
  n25 = TextureSample(UVs=Multiply(A=Frac(None=@n24), B=@n13.Result), Tex=@n14) {samplerType=SAMPLERTYPE_Masks, texture=&T_GridCell}
  n26 = Abs(None=LinearInterpolate(A=ComponentMask(None=@n21) {g=true}, B=ComponentMask(None=@n21) {r=true}, Alpha=@n3))
  n27 = LinearInterpolate(
   A=LinearInterpolate(A=@n15.RGB, B=@n20.RGB, Alpha=ComponentMask(None=@n23.Result) {r=true})
   B=@n25.RGB
   Alpha=ComponentMask(None=MaterialFunctionCall(In=@n26, Contrast=@n22) {materialFunction=&CheapContrast}.Result) {r=true}
  )
  n28 = OneMinus(None=@n27)
  n29 = ScalarParameter() {defaultValue=0.1, group=Equalizer, parameterName=EQ_Roughness}
  n30 = Saturate(None=MaterialFunctionCall(Float2=@n18) {materialFunction=&BreakOutFloat2Components}.G)
  n31 = ScalarParameter() {group=Equalizer, parameterName=GradientMin}
  n32 = ScalarParameter() {defaultValue=1, group=Equalizer, parameterName=GradientMax}
  n33 = ScalarParameter() {defaultValue=15, group=Equalizer, parameterName=EmoteColorGradientPow}
  n34 = CollectionParameter() {collection=&FortniteMaterialParameters, parameterName=Emote_AudioLow}
  n35 = CollectionParameter() {collection=&FortniteMaterialParameters, parameterName=Emote_AudioHigh}
  n36 = Add(A=@n34, B=@n35)
  n37 = CollectionParameter() {collection=&FortniteMaterialParameters, parameterName=EventAudio_Spectrum}
  n38 = MaterialFunctionCall(Float4=@n37) {materialFunction=&BreakOutFloat4Components}
  n39 = Add(A=Multiply(A=@n38.R) {constB=0.25}, B=Multiply(A=@n38.G) {constB=0.25})
  n40 = Multiply(A=@n38.B) {constB=0.25}
  n41 = Add(A=@n40, B=Multiply(A=@n38.A) {constB=0.25})
  n42 = CollectionParameter() {collection=&FortniteMaterialParameters, parameterName=EventAudio_Amplitude}
  n43 = Multiply(A=@n42) {constB=0.5}
  n44 = Power(Base=Add(A=Add(A=@n39, B=@n41), B=@n43))
  n45 = LinearInterpolate(
   A=LinearInterpolate(
    A=VectorParameter() {group=Equalizer, parameterName=ColorHigh, defaultValue=<rgba 0.1499 0.3059 0.3438 1>}.RGB
    B=VectorParameter() {group=Equalizer, parameterName=ColorLow, defaultValue=<rgba 0 0 0 1>}.RGB
    Alpha=MaterialFunctionCall(Alpha=@n30, Min=@n31, Max=@n32) {materialFunction=&SmoothStep}.Result
   )
   B=LinearInterpolate(
    A=VectorParameter() {group=Equalizer, parameterName=ColorEmoteHigh, defaultValue=<rgba 0.5 0 0 0>}.RGB
    B=VectorParameter() {group=Equalizer, parameterName=ColorEmoteLow, defaultValue=<rgba 0.5 0.3977 0 1>}.RGB
    Alpha=Power(Base=@n30, Exp=@n33)
   )
   Alpha=Power(Base=Max(A=@n36, B=@n44), Exp=ScalarParameter() {defaultValue=0.12, group=Equalizer, parameterName=EmoteThreshold})
  )
  n46 = MaterialFunctionCall(
   X=ScalarParameter() {defaultValue=5, group=Equalizer, parameterName=EQ_CellsX}
   Y=ScalarParameter() {defaultValue=8, group=Equalizer, parameterName=EQ_CellsY}
  ) {materialFunction=&MakeFloat2}
  n47 = Multiply(A=@n12, B=@n46.Result) {constB=5}
  n48 = MaterialFunctionCall(Float2=Divide(A=Floor(None=@n47), B=@n46.Result) {constB=5}) {materialFunction=&BreakOutFloat2Components}
  n49 = ScalarParameter() {group=Equalizer, parameterName=U_mul}
  MaterialExpressionMaterialFunctionCall_22 = MaterialFunctionCall(Float2=AppendVector(A=Multiply(A=@n48.R, B=@n49) {constB=0.5}, B=@n48.G)) {materialFunction=&BreakOutFloat2Components}
  n50 = Custom!(x=@MaterialExpressionMaterialFunctionCall_22.R) {
   code="return abs(2.*frac(x)-1.);"
   description=Mirror
   outputType=CMOT_Float1
   inputs [{
     inputName=x
     input {expression=&self:MaterialExpressionMaterialFunctionCall_22}
    }]
  }
  n51 = MaterialFunctionCall(
   X=ScalarParameter() {defaultValue=50, group=Equalizer, parameterName=EQ_CellsX}
   Y=ScalarParameter() {defaultValue=200, group=Equalizer, parameterName=EQ_CellsY}
  ) {materialFunction=&MakeFloat2}
  n52 = Multiply(A=@n18, B=@n51.Result) {constB=5}
  n53 = MaterialFunctionCall(Float2=Divide(A=Floor(None=@n52), B=@n51.Result) {constB=5}) {materialFunction=&BreakOutFloat2Components}
  n54 = ScalarParameter() {defaultValue=4, group=Equalizer, parameterName=U_mul}
  MaterialExpressionMaterialFunctionCall_19 = MaterialFunctionCall(Float2=AppendVector(A=Multiply(A=@n53.R, B=@n54) {constB=0.5}, B=@n53.G)) {materialFunction=&BreakOutFloat2Components}
  n55 = Custom(x=@MaterialExpressionMaterialFunctionCall_19.R) {
   code="return abs(2.*frac(x)-1.);"
   description=Mirror
   outputType=CMOT_Float1
   inputs [{
     inputName=x
     input {expression=&self:MaterialExpressionMaterialFunctionCall_19}
    }]
  }
  n56 = ComponentMask(None=@n21) {g=true}
  n57 = Abs(None=LinearInterpolate(A=ComponentMask(None=@n21) {b=true}, B=OneMinus(None=@n56), Alpha=@n3))
  n58 = LinearInterpolate(
   A=AppendVector(A=@n50, B=@MaterialExpressionMaterialFunctionCall_22.G)
   B=AppendVector(A=@n55, B=@MaterialExpressionMaterialFunctionCall_19.G)
   Alpha=ComponentMask(None=MaterialFunctionCall(In=@n57) {materialFunction=&CheapContrast}.Result) {r=true}
  )
  MaterialExpressionFrac_6 = Frac(None=@n58)
  n59 = Time()
  MaterialExpressionFrac_3 = Frac(None=Multiply(A=ScalarParameter() {defaultValue=2, group=Equalizer, parameterName=EQ_SpeedBass}, B=@n59))
  n60 = Custom!(InputFunction=@MaterialExpressionFrac_3) {
   code="float result = (2 * InputFunction) -2;\r\nresult  = result  * result  * result ;\r\nreturn -1.181 * InputFunction * result;"
   description=FastAttackSlowDecay
   outputType=CMOT_Float1
   inputs [{inputName=InputFunction, input {expression=&self:MaterialExpressionFrac_3}}]
  }
  n61 = Add(A=Multiply(A=@n38.R) {constB=0.5}, B=@n40)
  MaterialExpressionMax_2 = Max(
   A=Multiply(A=@n60, B=ScalarParameter() {defaultValue=0.4, group=Equalizer, parameterName=EQ_MaxIntensityBass})
   B=Max(A=@n34, B=Power(Base=@n61))
  )
  n62 = ScalarParameter() {defaultValue=0.5, group=Equalizer, parameterName=EQ_MaxIntensityTreble}
  MaterialExpressionFrac_4 = Frac(None=Multiply(A=@n59, B=ScalarParameter() {defaultValue=1.2, group=Equalizer, parameterName=EQ_SpeedTreble}) {desc="frac(time * speed)"})
  n63 = Custom(InputFunction=@MaterialExpressionFrac_4) {
   code="float result = (2 * InputFunction) -2;\r\nresult  = result  * result  * result ;\r\nreturn -1.181 * InputFunction * result;"
   description=FastAttackSlowDecay
   outputType=CMOT_Float1
   inputs [{inputName=InputFunction, input {expression=&self:MaterialExpressionFrac_4}}]
  }
  n64 = Power(Base=@n43)
  MaterialExpressionReroute_26 = Reroute(None=Max(A=Multiply(A=@n62, B=@n63), B=Max(A=@n35, B=@n64)))
  n65 = ComponentMask(None=@n58) {r=true}
  MaterialExpressionLinearInterpolate_0 = LinearInterpolate(A=@MaterialExpressionMax_2, B=@MaterialExpressionReroute_26, Alpha=@n65)
  MaterialExpressionLinearInterpolate_7 = LinearInterpolate(A=@MaterialExpressionMax_2, B=@MaterialExpressionReroute_26, Alpha=@n65)
  MaterialExpressionLinearInterpolate_8 = LinearInterpolate(A=@MaterialExpressionMax_2, B=@MaterialExpressionReroute_26, Alpha=@n65)
  MaterialExpressionScalarParameter_16 = ScalarParameter() {defaultValue=3.5, group=Equalizer, parameterName=EQ_BandsWidth}
  n66 = Custom!(
   uv=@MaterialExpressionFrac_6
   waveIntensity0=@MaterialExpressionMax_2
   waveIntensity1=@MaterialExpressionLinearInterpolate_0
   waveIntensity2=@MaterialExpressionLinearInterpolate_7
   waveIntensity3=@MaterialExpressionLinearInterpolate_8
   waveIntensity4=@MaterialExpressionReroute_26
   waveWidth=@MaterialExpressionScalarParameter_16
  ) {
   code="float waveA = -pow((waveWidth * uv.x - 0), 2) + waveIntensity0;\r\nfloat waveB = -pow((waveWidth * uv.x - waveWidth * 0.25), 2) + waveIntensity1;\r\nfloat waveC = -pow((waveWidth * uv.x - waveWidth * 0.5), 2) + waveIntensity2;\r\nfloat waveD = -pow((waveWidth * uv.x - waveWidth * 0.75), 2) + waveIntensity3;\r\nfloat waveE = -pow((waveWidth * uv.x - waveWidth), 2) + waveIntensity4;\r\nreturn max(waveA,max(waveB,max(waveC,max(waveD,waveE))));"
   description=CreateWave
   outputType=CMOT_Float1
   inputs [
    {inputName=uv, input {expression=&self:MaterialExpressionFrac_6}}
    {inputName=waveIntensity0, input {expression=&self:MaterialExpressionMax_2}}
    {
     inputName=waveIntensity1
     input {expression=&self:MaterialExpressionLinearInterpolate_0}
    }
    {
     inputName=waveIntensity2
     input {expression=&self:MaterialExpressionLinearInterpolate_7}
    }
    {
     inputName=waveIntensity3
     input {expression=&self:MaterialExpressionLinearInterpolate_8}
    }
    {
     inputName=waveIntensity4
     input {expression=&self:MaterialExpressionReroute_26}
    }
    {
     inputName=waveWidth
     input {expression=&self:MaterialExpressionScalarParameter_16}
    }
   ]
  }
  MaterialExpressionSaturate_7 = Saturate(None=@n66)
  MaterialExpressionLinearInterpolate_18 = LinearInterpolate(
   A=ScalarParameter() {defaultValue=0.8, group=Equalizer, parameterName=EQ_VerticalOffset}
   B=ScalarParameter() {defaultValue=0.8, group=Equalizer, parameterName=EQ_VerticalOffset_Vehicle}
   Alpha=@n3
  )
  n67 = Custom!(uv=@MaterialExpressionFrac_6, waveFunction=@MaterialExpressionSaturate_7, verticalOffset=@MaterialExpressionLinearInterpolate_18) {
   code="float result = uv.y- verticalOffset + waveFunction;    \r\nreturn (step(0.1, result));"
   description=DrawWave
   outputType=CMOT_Float1
   inputs [
    {inputName=uv, input {expression=&self:MaterialExpressionFrac_6}}
    {inputName=waveFunction, input {expression=&self:MaterialExpressionSaturate_7}}
    {
     inputName=verticalOffset
     input {expression=&self:MaterialExpressionLinearInterpolate_18}
    }
   ]
  }
  n68 = ScalarParameter() {defaultValue=1, group=Equalizer, parameterName=NoiseTiles}
  n69 = Time()
  n70 = ScalarParameter() {defaultValue=0.4, group=Equalizer, parameterName=NoiseSpeed}
  n71 = TextureSampleParameter2D(UVs=Add(A=Multiply(A=@n58, B=@n68), B=Multiply(A=@n69, B=@n70))) {group=Equalizer, parameterName=NoiseTexture, samplerType=SAMPLERTYPE_Masks, texture=/Game/Global/Textures/Default/Noise/T_CellularNoise}
  n72 = ComponentMask(None=@n71.RGB) {g=true, r=true}
  n73 = AppendVector(
   A=ScalarParameter() {defaultValue=0.7, group=Equalizer, parameterName=StaticPartsNoiseMin}
   B=ScalarParameter() {group=Equalizer, parameterName=EQ_NoiseMin}
  )
  n74 = ScalarParameter() {defaultValue=4, group=Equalizer, parameterName=StaticPartsNoisePow}
  n75 = ScalarParameter() {defaultValue=0.6, group=Equalizer, parameterName=EQ_NoisePow}
  n76 = MaterialFunctionCall(Float2=Power(Base=Max(A=@n72, B=@n73), Exp=AppendVector(A=@n74, B=@n75))) {materialFunction=&BreakOutFloat2Components}
  n77 = Multiply(B=@n67) {constA=1}
  n78 = LinearInterpolate(
   A=VectorParameter() {group=Equalizer, parameterName=BGColor}.RGB
   B=Multiply(A=Multiply(A=@n27, B=@n45), B=Multiply(A=@n67, B=@n76.G))
   Alpha=@n77
  )
  n79 = Multiply(A=ScalarParameter() {group=Equalizer, parameterName=EQ_Color_Intensity}, B=@n78) {constB=0.1}
  n80 = Multiply(A=@n28, B=VectorParameter() {group=Equalizer, parameterName=LinesColor}.RGB) {constB=0.5}
  n81 = TextureSampleParameterCube!(UVs=ReflectionVectorWS(), Apply View MipBias) {
   group=Equalizer
   parameterName=BackgroundCubemap
   samplerType=SAMPLERTYPE_LinearColor
   texture=/Game/Characters/Player/Male/Medium/Bodies/M_MED_DJ_Remix/Textures/TechArt_Textures/T_CoatCubeMap_Tex
  }
  n82 = Multiply(A=@n81.RGB, B=OneMinus(None=@n77))
  n83 = ScalarParameter() {defaultValue=0.1, group=Equalizer, parameterName=EQ_CubeMapIntensity}
  n84 = SetMaterialAttributes!(
   MaterialAttributes=@n2.MaterialAttributes
   BaseColor=Add(A=Add(A=@n79, B=@n80), B=Multiply(A=@n82, B=@n83) {constB=0.05}) {constB=0}
   EmissiveColor=Multiply(A=@n78, B=ScalarParameter() {defaultValue=15, group=Equalizer, parameterName=EQ_Emissive_Intensity})
   Normal=@n2.Normal
   Roughness=Saturate(None=Add(A=@n28, B=@n29))
  ) {attributeSetTypes [
    #2
    #3
    #1
    #4
   ]}
  n85 = FunctionInput(Preview=Constant() {r=1}) {bUsePreviewValueAsDefault=true, inputName="Customization Mask (_CM B)", inputType=FunctionInput_Scalar, sortPriority=24}
  FunctionOutput(None=StaticSwitchParameter(True=BlendMaterialAttributes(A=@n1, B=@n84, Alpha=@n85), False=@n1) {group=Customization_Wrap, parameterName=Equalizer})
  n86 = CollectionParameter() {collection=&FortniteMaterialParameters, parameterName=Emote_AudioHigh}
  n87 = ScalarParameter() {group=Equalizer, parameterName=Wrap_FFT_Magnitude}
  n88 = ScalarParameter() {group=Equalizer, parameterName=Use Wrap FFT}
  LinearInterpolate(A=@n86, B=@n87, Alpha=@n88)
  n89 = CollectionParameter() {collection=&FortniteMaterialParameters, parameterName=Emote_AudioLow}
  LinearInterpolate(A=@n89, B=@n87, Alpha=@n88)
  LinearInterpolate(A=Add(A=@n89, B=@n86), B=@n87, Alpha=@n88)
  n90 = Constant()
  LinearInterpolate(A=@n89, B=@n90, Alpha=@n88)
  LinearInterpolate(A=@n86, B=@n90, Alpha=@n88)
  MaterialFunctionCall(
   X=ScalarParameter() {defaultValue=100, group=Equalizer, parameterName=CellsX}
   Y=Divide(A=ScalarParameter() {defaultValue=200, group=Equalizer, parameterName=CellsY})
  ) {materialFunction=&MakeFloat2}
  n91 = Constant3Vector() {constant=<rgba 1 1 0.5 0>}
  Multiply(
   A=Add(
    A=Multiply(A=@n2.Normal, B=@n91)
    B=TextureSample(UVs=@n19) {samplerType=SAMPLERTYPE_Normal, texture=/Game/Characters/Player/Male/Medium/Bodies/M_MED_DJ_Remix/Textures/TechArt_Textures/T_Gem_N}.RGB
   )
   B=Constant3Vector() {constant=<rgba 1 1 0.675 0>}
  )
  OneMinus(None=@n8.B)
  VectorParameter() {group=Equalizer, parameterName=CubeMapColor, defaultValue=<rgba 0.7975 1 0.4754 0>}
 ]}
