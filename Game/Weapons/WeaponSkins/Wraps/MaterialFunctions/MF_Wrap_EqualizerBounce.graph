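# Material function for the EqualizerBounce wrap, layering a bobbing, rocking flipbook face, glowing equalizer bars and a hue-cycling accent over the incoming material.
# Side projections, tuned per item type, place the face and two mirrored equalizer mask samples.
# Four bar bands combine sine motion, MPC music spectrum/amplitude and timed decays from a Blueprint timestamp parameter.
# Input masks blend the layers; a static switch bypasses the wrap.
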MaterialFunction /Game/Weapons/WeaponSkins/Wraps/MaterialFunctions/MF_Wrap_EqualizerBounce
graph {nodes [
  n1 = MaterialFunctionCall(Float4=NamedRerouteUsage() {desc=" "}) {desc=" ", materialFunction=&BreakOutFloat4Components}
  n2 = Sine(None=Multiply(A=Time(), B=ScalarParameter() {group=Wrap_EqualizerBounce, parameterName=EquBounce_Face_Rotation_Speed}))
  n3 = ScalarParameter() {defaultValue=0.1, group=Wrap_EqualizerBounce, parameterName=EquBounce_Face_Rotation_Amount}
  n4 = Constant()
  n5 = ScalarParameter() {defaultValue=0.1, group=Wrap_EqualizerBounce_Audio, parameterName=EquBounce_Face_bobbing_Amount}
  n6 = CollectionParameter() {collection=/Game/Athena/Events/AudioAnalysis/MPC_Audio_Reactive_Parameters, parameterName=Spectrum_Music}
  n7 = Add(A=Saturate(None=ComponentMask(None=@n6) {b=true})) {constB=-0.35}
  n8 = Multiply(A=CollectionParameter() {collection=&MPC_Audio_Reactive_Parameters, parameterName=Amplitude_Music}, B=@n6)
  n9 = Step(X=Saturate(None=ComponentMask(None=@n8) {r=true})) {constX=0.25, constY=0.01}
  n10 = Add(
   A=LinearInterpolate(
    B=ScalarParameter() {defaultValue=0.1, group=Wrap_EqualizerBounce_Audio, parameterName=EquBounce_Face_Reactive_bobbing_Amount}
    Alpha=NamedRerouteUsage()
   ) {constB=0}
   B=LinearInterpolate(B=Multiply(A=@n5, B=@n7) {constB=0.1}, Alpha=Saturate(None=@n9))
  )
  n11 = MaterialFunctionCall!(
   For_Safe_Zones=StaticBool() {desc=" ", value=true}
   Normal_Map=NamedRerouteUsage() {desc=" "}
   Caculated_Bounds=NamedRerouteUsage() {desc=" "}
   Pre_Skinned_Normals=NamedRerouteUsage() {desc=" "}
   Is Vehicle=@n1.R
   Is VehicleSide=@n1.R
   Is Mech=@n1.R
   Vehicle_Rotation
   UV_Rotation Bypass=Add(A=NamedRerouteUsage() {desc=" "}, B=Multiply(A=@n2, B=@n3))
   Mech_Rotation
   Uvs_Offset=Add(
    A=NamedRerouteUsage() {desc=" "}
    B=MaterialFunctionCall(X=@n4, Y=@n10, Z=@n4) {materialFunction=&MakeFloat3}.Result
   )
   Offset-after rotation
   Uvs_TileRate=Multiply(A=NamedRerouteUsage() {desc=" "}, B=NamedRerouteUsage())
  ) {desc=" ", materialFunction=&/MF_Projection_Side_Normal}
  n12 = ScalarParameter() {defaultValue=3, group=Wrap_EqualizerBounce, parameterName="EquBounce_Flip_Rows\\Colums"}
  n13 = NamedRerouteUsage() {desc=" "}
  n14 = Divide(A=TextureProperty(Tex=@n13), B=Multiply(A=@n12, B=@n12))
  n15 = ScalarParameter() {defaultValue=1, group=Wrap_EqualizerBounce, parameterName=EquBounce_Flip_Miplevel_Adjustment, sortPriority=0}
  n16 = MaterialFunctionCall(
   "Animation  Phase (0-1)"=Divide(A=ScalarParameter() {group=Wrap_EqualizerBounce, parameterName=EquBounce_Flip_Frame}, B=Multiply(A=@n12, B=@n12)) {constB=1}
   Number of Rows=@n12
   Number of Columns=@n12
   Texture=@n13
   UVs=@n11.UVs
   MipBias/Level=MaterialFunctionCall!(UVs=@n11.UVs No Offsets/Tiling, Texture Size=Multiply(A=@n14, B=@n15)) {materialFunction=/Engine/Functions/Engine_MaterialFunctions01/Texturing/ComputeMipLevel}.MipLevel
  ) {materialFunction=&FlipBook}
  n17 = Multiply(A=NamedRerouteUsage(), B=Multiply(A=@n16.Alpha, B=@n11.Safe_Zone_Mask))
  Add(A=LinearInterpolate(B=@n11.Raw_Normal, Alpha=@n17), B=NamedRerouteUsage())
  Constant3Vector()
  Constant4Vector()
  n18 = TextureSample() {samplerType=SAMPLERTYPE_Masks, texture=/Game/Weapons/FORT_Rifles/Textures/T_Scar_CM}
  n19 = FunctionInput(Preview=@n18.R) {bUsePreviewValueAsDefault=true, inputName=Mask Primary, inputType=FunctionInput_Scalar, sortPriority=24, previewValue=<1 0 0 1>}
  n20 = FunctionInput(Preview=@n18.B) {bUsePreviewValueAsDefault=true, inputName=Mask Secondary, inputType=FunctionInput_Scalar, sortPriority=24, previewValue=<1 0 0 1>}
  n21 = BlendMaterialAttributes(A=NamedRerouteUsage(), B=NamedRerouteUsage(), Alpha=Saturate(None=Add(A=@n19, B=@n20))) {vertexAttributeBlendType=UseA}
  n22 = NamedRerouteUsage()
  n23 = NamedRerouteUsage()
  n24 = Saturate(None=Multiply(A=@n20, B=OneMinus(None=@n23)))
  n25 = TextureSample() {samplerType=SAMPLERTYPE_Normal, texture=/Game/Weapons/FORT_Rifles/T_SCAR_N}
  n26 = Constant3Vector() {constant=<rgba 1 0 0 0>}
  n27 = GetMaterialAttributes(None=FunctionInput(Preview=SetMaterialAttributes!(MaterialAttributes, Normal=@n25.RGB, BaseColor=@n26) {attributeSetTypes [#1, #2]}) {bUsePreviewValueAsDefault=true, inputName=Material, inputType=FunctionInput_MaterialAttributes}) {attributeGetTypes [
    #2
    #3
    #1
    #4
   ]}
  FunctionOutput(None=StaticSwitchParameter(
    True=BlendMaterialAttributes(A=@n21, B=@n22, Alpha=@n24) {vertexAttributeBlendType=UseA}
    False=NamedRerouteDeclaration(None=@n27.MaterialAttributes) {name=Root_Material, nodeColor=<rgba 0.7812 1 0 1>}
   ) {group=Customization_Wrap, parameterName=Wrap_EqualizerBounce})
  n28 = MaterialFunctionCall(Float4=NamedRerouteUsage()) {materialFunction=&BreakOutFloat4Components}
  n29 = NamedRerouteUsage()
  n30 = NamedRerouteUsage()
  n31 = Multiply(A=NamedRerouteUsage(), B=NamedRerouteUsage())
  n32 = MaterialFunctionCall(
   For_Safe_Zones=StaticBool() {value=true}
   Caculated_Bounds=NamedRerouteUsage()
   Pre_Skinned_Normals=NamedRerouteUsage()
   Is Vehicle=@n28.R
   Is VehicleSide=@n28.R
   Is Mech=@n28.R
   UV_Rotation Bypass=@n29
   Uvs_Offset=Add(A=NamedRerouteUsage(), B=@n30)
   Uvs_TileRate=@n31
  ) {materialFunction=&MF_Projection_Side_Normal}
  n33 = NamedRerouteUsage()
  n34 = MaterialFunctionCall(Float4=NamedRerouteUsage()) {materialFunction=&BreakOutFloat4Components}
  n35 = Constant3Vector() {constant=<rgba -1 -1 1 0>}
  n36 = MaterialFunctionCall(
   For_Safe_Zones=StaticBool() {value=true}
   Caculated_Bounds=NamedRerouteUsage()
   Pre_Skinned_Normals=NamedRerouteUsage()
   Is Vehicle=@n34.R
   Is VehicleSide=@n34.R
   Is Mech=@n34.R
   UV_Rotation Bypass=@n29
   Uvs_Offset=Add(A=NamedRerouteUsage(), B=Multiply(A=@n30, B=@n35))
   Uvs_TileRate=Multiply(A=@n31, B=Constant3Vector() {constant=<rgba -1 1 1 0>})
  ) {materialFunction=&MF_Projection_Side_Normal}
  n37 = NamedRerouteUsage()
  n38 = Saturate(None=Add(
    A=TextureSample(UVs=@n32.UVs, Tex=@n33) {samplerType=SAMPLERTYPE_Masks}.A
    B=TextureSample(UVs=@n36.UVs, Tex=@n37) {samplerType=SAMPLERTYPE_Masks}.A
   ))
  n39 = TextureSample(UVs=@n32.UVs, Tex=NamedRerouteUsage()) {samplerType=SAMPLERTYPE_Masks}
  n40 = TextureSample(UVs=@n36.UVs, Tex=NamedRerouteUsage()) {samplerType=SAMPLERTYPE_Masks}
  n41 = Add(A=@n39.R, B=@n40.R)
  n42 = Add(A=@n39.B, B=@n40.B)
  n43 = Min(A=@n41, B=@n42)
  Multiply(A=@n38, B=@n43)
  n44 = VectorParameter() {group=Wrap_EqualizerBounce_UVs, parameterName=EquBounce_Equlizer_Scale, defaultValue=<rgba 2.5 1.75 1 2>}
  Multiply(A=@n44.RGB, B=@n44.RGB)
  n45 = Multiply(A=Time(), B=ScalarParameter() {defaultValue=0.25, group=Wrap_EqualizerBounce, parameterName=EquBounce_Face_bobbing_Speed})
  n46 = Multiply(A=Add(A=Sine(None=@n45))) {constB=0.5}
  Multiply(A=Add(
    A=MaterialFunctionCall!(X range Y range, Gradient Height, ShowGrid?, Thickness, Color, "f(x)"=@n46) {materialFunction=/Engine/Functions/Engine_MaterialFunctions01/Debug/PlotFunctionOnGraph}."f(x) = y"
    B=MaterialFunctionCall() {materialFunction=&PlotFunctionOnGraph}."f(x) = y"
   ))
  n47 = ScalarParameter() {defaultValue=1, group=Wrap_EqualizerBounce_UVs, parameterName=EquBounce_Scale_Vehicle, sortPriority=150}
  n48 = ScalarParameter() {defaultValue=1, group=Wrap_EqualizerBounce_UVs, parameterName=EquBounce_Scale_Vehicle_Side, sortPriority=150}
  n49 = MaterialFunctionCall(Float4=NamedRerouteUsage()) {materialFunction=&BreakOutFloat4Components}
  n50 = LinearInterpolate(
   A=LinearInterpolate(
    A=ScalarParameter() {defaultValue=0.1, group=Wrap_EqualizerBounce_UVs, parameterName=EquBounce_Scale_Weapon, sortPriority=150}
    B=LinearInterpolate(A=@n47, B=@n48, Alpha=@n49.G)
    Alpha=@n49.R
   )
   B=LinearInterpolate(
    A=ScalarParameter() {defaultValue=1, group=Wrap_EqualizerBounce_UVs, parameterName=EquBounce_Scale_Mech, sortPriority=150}
    B=ScalarParameter() {defaultValue=1, group=Wrap_EqualizerBounce_UVs, parameterName=EquBounce_Scale_Character, sortPriority=150}
    Alpha=@n49.A
   )
   Alpha=@n49.B
  )
  n51 = ScalarParameter() {defaultValue=1, group=Wrap_EqualizerBounce_UVs, parameterName=EquBounce_Scale_Backpack, sortPriority=150}
  n52 = MaterialFunctionCall(Float2=NamedRerouteUsage()) {materialFunction=&BreakOutFloat2Components}
  NamedRerouteDeclaration(None=LinearInterpolate(
    A=LinearInterpolate(A=@n50, B=@n51, Alpha=@n52.R)
    B=ScalarParameter() {defaultValue=1, group=Wrap_EqualizerBounce_UVs, parameterName=EquBounce_Scale_Pickaxe, sortPriority=150}
    Alpha=@n52.G
   )) {name=UVs_Scale, nodeColor=<rgba 0 0.7188 1 1>}
  n53 = VectorParameter() {group=Wrap_EqualizerBounce_UVs, parameterName=EquBounce_Offsest_Weapon, defaultValue=<rgba -0.35 -0.01 0 1>}
  n54 = VectorParameter() {group=Wrap_EqualizerBounce_UVs, parameterName=EquBounce_Offsest_Vehicle, defaultValue=<rgba 0 0 0 1>}
  n55 = VectorParameter() {group=Wrap_EqualizerBounce_UVs, parameterName=EquBounce_Offsest_VehicleSide, defaultValue=<rgba 0 0 0 1>}
  n56 = MaterialFunctionCall(Float4=NamedRerouteUsage()) {materialFunction=&BreakOutFloat4Components}
  n57 = LinearInterpolate(
   A=LinearInterpolate(A=@n53.RGB, B=LinearInterpolate(A=@n54.RGB, B=@n55.RGB, Alpha=@n56.G), Alpha=@n56.R)
   B=LinearInterpolate(
    A=VectorParameter() {group=Wrap_EqualizerBounce_UVs, parameterName=EquBounce_Offsest_Mech, defaultValue=<rgba 0 0 0 1>}.RGB
    B=VectorParameter() {group=Wrap_EqualizerBounce_UVs, parameterName=EquBounce_Offsest_Character, defaultValue=<rgba 0 0 0 1>}.RGB
    Alpha=@n56.A
   )
   Alpha=@n56.B
  )
  n58 = VectorParameter() {group=Wrap_EqualizerBounce_UVs, parameterName=EquBounce_Offsest_Backpack, defaultValue=<rgba 0 0 0 1>}
  n59 = MaterialFunctionCall(Float2=NamedRerouteUsage()) {materialFunction=&BreakOutFloat2Components}
  NamedRerouteDeclaration(None=LinearInterpolate(
    A=LinearInterpolate(A=@n57, B=@n58.RGB, Alpha=@n59.R)
    B=VectorParameter() {group=Wrap_EqualizerBounce_UVs, parameterName=EquBounce_Offsest_Pickaxe, defaultValue=<rgba 0 0 0 1>}.RGB
    Alpha=@n59.G
   )) {name=UVs_Offset, nodeColor=<rgba 0 0.625 1 1>}
  NamedRerouteDeclaration(None=TextureObjectParameter() {group=Wrap_EqualizerBounce, parameterName=EquBounce__Tex__Msk, texture=&T_BlacksRGB}) {name=Tex_Msk, nodeColor=<rgba 0.4531 0 1 1>}
  n60 = MaterialFunctionCall(Float4=NamedRerouteUsage()) {materialFunction=&BreakOutFloat4Components}
  NamedRerouteDeclaration(None=MaterialFunctionCall(
    Normal_Map=NamedRerouteUsage()
    Caculated_Bounds=NamedRerouteUsage()
    Pre_Skinned_Normals=NamedRerouteUsage()
    Is Vehicle=@n60.R
    Is VehicleSide=@n60.R
    Is Mech=@n60.R
    UV_Rotation Bypass=NamedRerouteUsage()
    Uvs_TileRate=Constant() {r=1}
   ) {materialFunction=&MF_Projection_Side_Normal}.UVs) {name=UVs, nodeColor=<rgba 1 0 0.6094 1>}
  n61 = ScalarParameter() {defaultValue=0.3, group=Wrap_EqualizerBounce, parameterName=EquBounce_Roughness}
  n62 = ScalarParameter() {defaultValue=0.2, group=Wrap_EqualizerBounce, parameterName=EquBounce_Roughness_Equlizer}
  n63 = Multiply(A=NamedRerouteUsage(), B=@n38)
  n64 = ScalarParameter() {defaultValue=0.1, group=Wrap_EqualizerBounce, parameterName=EquBounce_Specular}
  n65 = ScalarParameter() {defaultValue=0.25, group=Wrap_EqualizerBounce, parameterName=EquBounce_Specular_Equlizer}
  n66 = Time()
  n67 = VectorParameter() {group=Wrap_EqualizerBounce_Audio, parameterName=EquBounce_Equlizer_Speed, defaultValue=<rgba 0.6 0.65 0.7 0.8>}
  n68 = Add(A=Sine(None=Multiply(A=@n66, B=@n67.R)))
  n69 = VectorParameter() {group=Wrap_EqualizerBounce_Audio, parameterName=EquBounce_Equlizer_Multplyer, defaultValue=<rgba 0.1 0.1 0.15 0.19>}
  n70 = MaterialFunctionCall(Float4=Multiply(A=@n8, B=ScalarParameter() {defaultValue=1.25, group=Wrap_EqualizerBounce_Audio, parameterName=EquBounce_Equlizer_Reactive_Strength})) {materialFunction=&BreakOutFloat4Components}
  n71 = Subtract(
   A=Time()
   B=Add(
    A=ScalarParameter() {defaultValue=-9999999, group=Wrap_TimefromBP, parameterName=TimeFrom-BP----------------, sortPriority=600}
    B=ScalarParameter() {defaultValue=0.2, group=Wrap_EqualizerBounce_Audio, parameterName=EquBounce_Time_Delay, sortPriority=601}
   )
  )
  n72 = ScalarParameter() {defaultValue=0.7, group=Wrap_EqualizerBounce_Audio, parameterName=EquBounce_Time_Duration_A, sortPriority=601}
  n73 = OneMinus(None=Saturate(None=Divide(A=@n71, B=@n72)))
  n74 = Add(A=Max(A=Multiply(A=@n68, B=@n69.R) {constB=0.5}, B=Max(A=@n70.A, B=@n73)), B=Add(A=@n41) {constB=-1})
  n75 = Step(X=@n43) {constX=0, constY=0.1}
  n76 = Saturate(None=Subtract(A=Step(X=@n41) {constX=0, constY=0.1}, B=@n75))
  n77 = Time()
  n78 = Add(A=Sine(None=Multiply(A=@n77, B=@n67.G)))
  n79 = ScalarParameter() {defaultValue=0.4, group=Wrap_EqualizerBounce_Audio, parameterName=EquBounce_Time_Duration_B, sortPriority=601}
  n80 = OneMinus(None=Saturate(None=Divide(A=@n71, B=@n79)))
  n81 = Add(A=@n39.G, B=@n40.G)
  n82 = Add(A=Max(A=Multiply(A=@n78, B=@n69.G) {constB=0.5}, B=Max(A=@n70.B, B=@n80)), B=Add(B=@n81) {constA=-1})
  n83 = Saturate(None=Step(X=@n81) {constX=0, constY=0.1})
  n84 = Time()
  n85 = Add(A=Sine(None=Multiply(A=@n84, B=@n67.B)))
  n86 = ScalarParameter() {defaultValue=0.5, group=Wrap_EqualizerBounce_Audio, parameterName=EquBounce_Time_Duration_C, sortPriority=601}
  n87 = OneMinus(None=Saturate(None=Divide(A=@n71, B=@n86)))
  n88 = Add(A=Max(A=Multiply(A=@n85, B=@n69.B) {constB=0.5}, B=Max(A=@n70.G, B=@n87)), B=Add(B=@n42) {constA=-1})
  n89 = Saturate(None=Subtract(A=Step(X=@n42) {constX=0, constY=0.1}, B=@n75))
  n90 = Time()
  n91 = Add(A=Sine(None=Multiply(A=@n90, B=@n67.A)))
  n92 = ScalarParameter() {defaultValue=0.8, group=Wrap_EqualizerBounce_Audio, parameterName=EquBounce_Time_Duration_D, sortPriority=601}
  n93 = OneMinus(None=Saturate(None=Divide(A=@n71, B=@n92)))
  n94 = Add(A=Max(A=Multiply(A=@n91, B=@n69.A) {constB=0.5}, B=Max(A=@n70.R, B=@n93)), B=Add(B=@n43) {constA=-1})
  n95 = Saturate(None=@n75)
  n96 = Add(A=Add(A=Multiply(A=@n74, B=@n76), B=Multiply(A=@n82, B=@n83)), B=Add(A=Multiply(A=@n88, B=@n89), B=Multiply(A=@n94, B=@n95)))
  n97 = Max(A=Max(A=@n41, B=@n81), B=@n42)
  n98 = Multiply(A=Time(), B=ScalarParameter() {defaultValue=0.1, group=Wrap_EqualizerBounce, parameterName=EquBounce_Equlizer_Color_Speed, sortPriority=150})
  n99 = Desaturation(
   None=MaterialFunctionCall(Hue Shift Percentage=Add(A=@n97, B=@n98) {constB=0.5}) {materialFunction=&HueShift}.Result
   Fraction=ScalarParameter() {defaultValue=0.5, group=Wrap_EqualizerBounce, parameterName=EquBounce_Equlizer_Color_Saturation, sortPriority=150}
  )
  n100 = Multiply(
   A=ScalarParameter() {defaultValue=3, group=Wrap_EqualizerBounce, parameterName=EquBounce_Equlizer_Color_Brightness, sortPriority=150}
   B=Multiply(A=Saturate(None=@n96), B=@n99)
  )
  n101 = Multiply(
   A=@n99
   B=ScalarParameter() {defaultValue=0.02, group=Wrap_EqualizerBounce, parameterName=EquBounce_Equlizer_Color_Bck_Brightness, sortPriority=150}
  ) {constB=0.005}
  n102 = VectorParameter() {group=Wrap_EqualizerBounce, parameterName=EquBounce_Face_Tint, defaultValue=<rgba 1 1 1 1>}
  n103 = SetMaterialAttributes!(
   MaterialAttributes=NamedRerouteUsage()
   BaseColor=LinearInterpolate(
    A=VectorParameter() {group=Wrap_EqualizerBounce, parameterName=EquBounce_BaseColor, defaultValue=<rgba 0.02 0.02 0.02 1>}.RGB
    B=Multiply(A=@n102.RGB, B=@n16.Result)
    Alpha=@n17
   )
   Roughness=LinearInterpolate(
    A=LinearInterpolate(A=@n61, B=@n62, Alpha=@n63)
    B=ScalarParameter() {defaultValue=0.35, group=Wrap_EqualizerBounce, parameterName=EquBounce_Roughness_Face}
    Alpha=NamedRerouteUsage()
   )
   Specular=LinearInterpolate(
    A=LinearInterpolate(A=@n64, B=@n65, Alpha=@n63)
    B=ScalarParameter() {defaultValue=0.5, group=Wrap_EqualizerBounce, parameterName=EquBounce_Specular_Face}
    Alpha=NamedRerouteUsage()
   )
   EmissiveColor=Multiply(A=Add(A=@n100, B=@n101), B=@n63)
   Metallic=ScalarParameter() {group=Wrap_EqualizerBounce, parameterName=EquBounce_Metalic}
   Normal
  ) {attributeSetTypes [
    #2
    #3
    #5
    #6
    #4
    #1
   ]}
  NamedRerouteDeclaration(None=@n103) {name=Base_Material, nodeColor=<rgba 1 0.7031 0 1>}
  NamedRerouteDeclaration(None=FunctionInput(Preview=Constant4Vector() {constant=<rgba -0.25 0 0 0.2>}.RGBA) {bUsePreviewValueAsDefault=true, inputName=Object-SafeZone-Front, inputType=FunctionInput_Vector4, sortPriority=102}) {name=SafeZone_Frnt, nodeColor=<rgba 0.9922 1 0 1>}
  n104 = ScalarParameter() {group=Wrap_EqualizerBounce_UVs, parameterName=EquBounce_Equalizer_Rotation_Vehicle, sortPriority=150}
  n105 = ScalarParameter() {group=Wrap_EqualizerBounce_UVs, parameterName=EquBounce_Equalizer_Rotation_Vehicle_Side, sortPriority=150}
  n106 = MaterialFunctionCall(Float4=NamedRerouteUsage()) {materialFunction=&BreakOutFloat4Components}
  n107 = LinearInterpolate(
   A=LinearInterpolate(
    A=ScalarParameter() {group=Wrap_EqualizerBounce_UVs, parameterName=EquBounce_Equalizer_Rotation_Weapon, sortPriority=150}
    B=LinearInterpolate(A=@n104, B=@n105, Alpha=@n106.G)
    Alpha=@n106.R
   )
   B=LinearInterpolate(
    A=ScalarParameter() {group=Wrap_EqualizerBounce_UVs, parameterName=EquBounce_Equalizer_Rotation_Mech, sortPriority=150}
    B=ScalarParameter() {group=Wrap_EqualizerBounce_UVs, parameterName=EquBounce_Equalizer_Rotation_Character, sortPriority=150}
    Alpha=@n106.A
   )
   Alpha=@n106.B
  )
  n108 = ScalarParameter() {group=Wrap_EqualizerBounce_UVs, parameterName=EquBounce_Equalizer_Rotation_Backpack, sortPriority=150}
  n109 = MaterialFunctionCall(Float2=NamedRerouteUsage()) {materialFunction=&BreakOutFloat2Components}
  NamedRerouteDeclaration(None=LinearInterpolate(
    A=LinearInterpolate(A=@n107, B=@n108, Alpha=@n109.R)
    B=ScalarParameter() {group=Wrap_EqualizerBounce_UVs, parameterName=EquBounce_Equalizer_Rotation_Pickaxe, sortPriority=150}
    Alpha=@n109.G
   )) {name=UVs_Rotation_Equ, nodeColor=<rgba 0 0.7188 1 1>}
  n110 = VectorParameter() {group=Wrap_EqualizerBounce_UVs, parameterName=EquBounce_Equalizer_Offset_Weapon, defaultValue=<rgba 0.29 0 0 1>}
  n111 = VectorParameter() {group=Wrap_EqualizerBounce_UVs, parameterName=EquBounce_Equalizer_Offset_Vehicle, defaultValue=<rgba 0 0 0 1>}
  n112 = VectorParameter() {group=Wrap_EqualizerBounce_UVs, parameterName=EquBounce_Equalizer_Offset_VehicleSide, defaultValue=<rgba 0 0 0 1>}
  n113 = MaterialFunctionCall(Float4=NamedRerouteUsage()) {materialFunction=&BreakOutFloat4Components}
  n114 = VectorParameter() {group=Wrap_EqualizerBounce_UVs, parameterName=EquBounce_Equalizer_Offset_Mech, defaultValue=<rgba 0 0 0 1>}
  n115 = VectorParameter() {group=Wrap_EqualizerBounce_UVs, parameterName=EquBounce_Equalizer_Offset_Character, defaultValue=<rgba 0 0 0 1>}
  n116 = LinearInterpolate(
   A=LinearInterpolate(A=@n110.RGB, B=LinearInterpolate(A=@n111.RGB, B=@n112.RGB, Alpha=@n113.G), Alpha=@n113.R)
   B=LinearInterpolate(A=@n114.RGB, B=@n115.RGB, Alpha=@n113.A)
   Alpha=@n113.B
  )
  n117 = VectorParameter() {group=Wrap_EqualizerBounce_UVs, parameterName=EquBounce_Equalizer_Offset_Backpack, defaultValue=<rgba 0 0 0 1>}
  n118 = MaterialFunctionCall(Float2=NamedRerouteUsage()) {materialFunction=&BreakOutFloat2Components}
  n119 = VectorParameter() {group=Wrap_EqualizerBounce_UVs, parameterName=EquBounce_Equalizer_Offset_Pickaxe, defaultValue=<rgba 0 0 0 1>}
  NamedRerouteDeclaration(None=LinearInterpolate(A=LinearInterpolate(A=@n116, B=@n117.RGB, Alpha=@n118.R), B=@n119.RGB, Alpha=@n118.G)) {name=UVs_Equlizer_Offset, nodeColor=<rgba 0 0.625 1 1>}
  n120 = VectorParameter() {group=Wrap_EqualizerBounce_UVs, parameterName=EquBounce_Equalizer_Scale_Weapon, defaultValue=<rgba 2.5 1.75 1 2>}
  n121 = VectorParameter() {group=Wrap_EqualizerBounce_UVs, parameterName=EquBounce_Equalizer_Scale_Vehicle, defaultValue=<rgba 1 1 1 1>}
  n122 = Multiply(A=@n121.RGB, B=@n121.RGB)
  n123 = VectorParameter() {group=Wrap_EqualizerBounce_UVs, parameterName=EquBounce_Equalizer_Scale_VehicleSide, defaultValue=<rgba 1 1 1 1>}
  n124 = Multiply(A=@n123.RGB, B=@n123.RGB)
  n125 = MaterialFunctionCall(Float4=NamedRerouteUsage()) {materialFunction=&BreakOutFloat4Components}
  n126 = VectorParameter() {group=Wrap_EqualizerBounce_UVs, parameterName=EquBounce_Equalizer_Scale_Mech, defaultValue=<rgba 1 1 1 1>}
  n127 = VectorParameter() {group=Wrap_EqualizerBounce_UVs, parameterName=EquBounce_Equalizer_Scale_Character, defaultValue=<rgba 1 1 1 1>}
  n128 = LinearInterpolate(
   A=LinearInterpolate(A=Multiply(A=@n120.RGB, B=@n120.RGB), B=LinearInterpolate(A=@n122, B=@n124, Alpha=@n125.G), Alpha=@n125.R)
   B=LinearInterpolate(A=Multiply(A=@n126.RGB, B=@n126.RGB), B=Multiply(A=@n127.RGB, B=@n127.RGB), Alpha=@n125.A)
   Alpha=@n125.B
  )
  n129 = VectorParameter() {group=Wrap_EqualizerBounce_UVs, parameterName=EquBounce_Equalizer_Scale_Backpack, defaultValue=<rgba 1 1 1 1>}
  n130 = Multiply(A=@n129.RGB, B=@n129.RGB)
  n131 = MaterialFunctionCall(Float2=NamedRerouteUsage()) {materialFunction=&BreakOutFloat2Components}
  n132 = VectorParameter() {group=Wrap_EqualizerBounce_UVs, parameterName=EquBounce_Equalizer_Scale_Pickaxe, defaultValue=<rgba 1 1 1 1>}
  NamedRerouteDeclaration(None=LinearInterpolate(A=LinearInterpolate(A=@n128, B=@n130, Alpha=@n131.R), B=Multiply(A=@n132.RGB, B=@n132.RGB), Alpha=@n131.G)) {name=UVs_Equlizer_Scale, nodeColor=<rgba 0 0.625 1 1>}
  NamedRerouteDeclaration(None=@n17) {name=Face_Msk, nodeColor=<rgba 1 0 0.2812 1>}
  n133 = Divide(A=@n71, B=ScalarParameter() {defaultValue=0.1, group=Wrap_EqualizerBounce_Audio, parameterName=EquBounce_Time_Bounce, sortPriority=601})
  NamedRerouteDeclaration(None=OneMinus(None=Saturate(None=@n133))) {name=Bounce, nodeColor=<rgba 0 0.1094 1 1>}
  n134 = ScalarParameter() {group=Wrap_EqualizerBounce_UVs, parameterName=EquBounce_Rotation_Vehicle, sortPriority=150}
  n135 = ScalarParameter() {group=Wrap_EqualizerBounce_UVs, parameterName=EquBounce_Rotation_Vehicle_Side, sortPriority=150}
  n136 = MaterialFunctionCall(Float4=NamedRerouteUsage()) {materialFunction=&BreakOutFloat4Components}
  n137 = LinearInterpolate(
   A=LinearInterpolate(
    A=ScalarParameter() {group=Wrap_EqualizerBounce_UVs, parameterName=EquBounce_Rotation_Weapon, sortPriority=150}
    B=LinearInterpolate(A=@n134, B=@n135, Alpha=@n136.G)
    Alpha=@n136.R
   )
   B=LinearInterpolate(
    A=ScalarParameter() {group=Wrap_EqualizerBounce_UVs, parameterName=EquBounce_Rotation_Mech, sortPriority=150}
    B=ScalarParameter() {group=Wrap_EqualizerBounce_UVs, parameterName=EquBounce_Rotation_Character, sortPriority=150}
    Alpha=@n136.A
   )
   Alpha=@n136.B
  )
  n138 = ScalarParameter() {group=Wrap_EqualizerBounce_UVs, parameterName=EquBounce_Rotation_Backpack, sortPriority=150}
  n139 = MaterialFunctionCall(Float2=NamedRerouteUsage()) {materialFunction=&BreakOutFloat2Components}
  NamedRerouteDeclaration(None=LinearInterpolate(
    A=LinearInterpolate(A=@n137, B=@n138, Alpha=@n139.R)
    B=ScalarParameter() {group=Wrap_EqualizerBounce_UVs, parameterName=EquBounce_Rotation_Pickaxe, sortPriority=150}
    Alpha=@n139.G
   )) {name=UVs_Rotation, nodeColor=<rgba 0 0.7188 1 1>}
  n140 = MaterialFunctionCall(Color=Abs(None=NamedRerouteUsage())) {materialFunction=&SplitComponents}
  n141 = MaterialFunctionCall(Float4=NamedRerouteUsage()) {materialFunction=&BreakOutFloat4Components}
  n142 = LinearInterpolate(A=@n140.R, B=LinearInterpolate(A=@n140.B, B=@n140.B, Alpha=@n141.G), Alpha=@n141.R)
  NamedRerouteDeclaration(None=MaterialFunctionCall(
    In=LinearInterpolate(A=@n142, B=@n140.G, Alpha=@n141.B)
    Contrast=ScalarParameter() {defaultValue=1, group=_Base, parameterName=EquBounce_Blend_Contrast}
   ) {materialFunction=&CheapContrast}.Result) {name=Tri_Mask_1, nodeColor=<rgba 1 0.4688 0 1>}
  n143 = ScalarParameter() {defaultValue=1, group=Wrap_EqualizerBounce_UVs, parameterName=EquBounce_Face_Scale_Vehicle, sortPriority=150}
  n144 = ScalarParameter() {defaultValue=1, group=Wrap_EqualizerBounce_UVs, parameterName=EquBounce_Face_Scale_Vehicle_Side, sortPriority=150}
  n145 = MaterialFunctionCall(Float4=NamedRerouteUsage()) {materialFunction=&BreakOutFloat4Components}
  n146 = LinearInterpolate(
   A=LinearInterpolate(
    A=ScalarParameter() {defaultValue=1, group=Wrap_EqualizerBounce_UVs, parameterName=EquBounce_Face_Scale_Weapon, sortPriority=150}
    B=LinearInterpolate(A=@n143, B=@n144, Alpha=@n145.G)
    Alpha=@n145.R
   )
   B=LinearInterpolate(
    A=ScalarParameter() {defaultValue=1, group=Wrap_EqualizerBounce_UVs, parameterName=EquBounce_Face_Scale_Mech, sortPriority=150}
    B=ScalarParameter() {defaultValue=1, group=Wrap_EqualizerBounce_UVs, parameterName=EquBounce_Face_Scale_Character, sortPriority=150}
    Alpha=@n145.A
   )
   Alpha=@n145.B
  )
  n147 = ScalarParameter() {defaultValue=1, group=Wrap_EqualizerBounce_UVs, parameterName=EquBounce_Face_Scale_Backpack, sortPriority=150}
  n148 = MaterialFunctionCall(Float2=NamedRerouteUsage()) {materialFunction=&BreakOutFloat2Components}
  NamedRerouteDeclaration(None=LinearInterpolate(
    A=LinearInterpolate(A=@n146, B=@n147, Alpha=@n148.R)
    B=ScalarParameter() {defaultValue=1, group=Wrap_EqualizerBounce_UVs, parameterName=EquBounce_Face_Scale_Pickaxe, sortPriority=150}
    Alpha=@n148.G
   )) {name=UVs_Scale_Face, nodeColor=<rgba 0 0.7188 1 1>}
  n149 = MaterialFunctionCall(Color=Abs(None=NamedRerouteUsage())) {materialFunction=&SplitComponents}
  n150 = MaterialFunctionCall(Float4=NamedRerouteUsage()) {materialFunction=&BreakOutFloat4Components}
  n151 = LinearInterpolate(A=@n149.R, B=LinearInterpolate(A=@n149.B, B=@n149.B, Alpha=@n150.G), Alpha=@n150.R)
  NamedRerouteDeclaration(None=MaterialFunctionCall(
    In=LinearInterpolate(A=@n151, B=@n149.G, Alpha=@n150.B)
    Contrast=ScalarParameter() {defaultValue=1, group=Wrap_EqualizerBounce, parameterName=EquBounce_TriPlanerMsk_Contrast, sortPriority=9}
   ) {materialFunction=&CheapContrast}.Result) {name=Tri_Mask, nodeColor=<rgba 1 0.4688 0 1>}
  NamedRerouteDeclaration(None=TextureObjectParameter() {group=Wrap_EqualizerBounce, parameterName=EquBounce__Tex__Face, texture=&T_BlacksRGB}) {name=Tex_Face, nodeColor=<rgba 0.4531 0 1 1>}
  n152 = NamedRerouteUsage()
  n153 = MaterialFunctionCall(Float4=NamedRerouteUsage()) {materialFunction=&BreakOutFloat4Components}
  n154 = Max(
   A=@n153.A
   B=MaterialFunctionCall(Float2=NamedRerouteUsage()) {materialFunction=&BreakOutFloat2Components}.R
  )
  n155 = Multiply(
   A=LinearInterpolate(A=ComponentMask(None=@n152) {r=true}, B=ComponentMask(None=@n152) {g=true}, Alpha=Max(A=@n153.B, B=@n154))
   B=ScalarParameter() {defaultValue=1, group=Wrap_EqualizerBounce_Accent, parameterName=EquBounce_Accent_Color_TileRate, sortPriority=150}
  )
  n156 = Multiply(
   A=Time()
   B=ScalarParameter() {defaultValue=0.05, group=Wrap_EqualizerBounce_Accent, parameterName=EquBounce_Accent_Color_Speed, sortPriority=150}
  )
  n157 = MaterialFunctionCall(Hue Shift Percentage=Frac(None=Add(A=@n155, B=@n156) {constB=0.5}), Texture=Constant3Vector() {constant=<rgba 1 0 0 0>}) {materialFunction=&HueShift}
  n158 = ScalarParameter() {defaultValue=0.3, group=Wrap_EqualizerBounce_Accent, parameterName=EquBounce_Accent_Color_Saturation, sortPriority=150}
  n159 = SetMaterialAttributes(
   MaterialAttributes=NamedRerouteUsage()
   BaseColor=Multiply(
    A=Desaturation(None=@n157.Result, Fraction=@n158)
    B=ScalarParameter() {defaultValue=1, group=Wrap_EqualizerBounce_Accent, parameterName=EquBounce_Accent_Color_Brightness, sortPriority=150}
   )
   Roughness=ScalarParameter() {defaultValue=0.35, group=Wrap_EqualizerBounce_Accent, parameterName=EquBounce_Accent_Roughness, sortPriority=10}
   Specular=ScalarParameter() {defaultValue=0.5, group=Wrap_EqualizerBounce_Accent, parameterName=EquBounce_Accent_Specular, sortPriority=10}
   Metallic=ScalarParameter() {group=Wrap_EqualizerBounce_Accent, parameterName=EquBounce_Accent_Metalic, sortPriority=10}
  ) {attributeSetTypes [
    #2
    #3
    #5
    #6
    #4
    #1
   ]}
  NamedRerouteDeclaration(None=@n159) {name=Accent_Material, nodeColor=<rgba 1 0 0.6328 1>}
  n160 = FunctionInput(Preview=@n18.G) {bUsePreviewValueAsDefault=true, inputName=Mask Crevise, inputType=FunctionInput_Scalar, sortPriority=24, previewValue=<1 0 0 1>}
  NamedRerouteDeclaration(None=@n160) {name=Crevise_map, nodeColor=<rgba 1 0.6797 0 1>}
  NamedRerouteDeclaration(None=FunctionInput(Preview=Constant4Vector().RGBA) {bUsePreviewValueAsDefault=true, inputName=IsState, inputType=FunctionInput_Vector4, sortPriority=4}) {name=IsState, nodeColor=<rgba 0.1497 1 0.1633 1>}
  NamedRerouteDeclaration(None=FunctionInput(Preview=Constant2Vector()) {bUsePreviewValueAsDefault=true, inputName=IsState_B, inputType=FunctionInput_Vector2, sortPriority=4}) {name=IsState_B, nodeColor=<rgba 0.1497 1 0.1633 1>}
  n161 = PreSkinnedLocalBounds()
  n162 = Subtract(A=VertexInterpolator(VS=PreSkinnedPosition()).PS, B=@n161.Min)
  n163 = MaterialFunctionCall(Color=Subtract(A=@n161.Max, B=@n161.Max)) {materialFunction=&SplitComponents}
  n164 = Max(A=Max(A=@n163.R, B=@n163.R), B=@n163.B)
  NamedRerouteDeclaration(None=FunctionInput(Preview=Divide(A=@n162, B=@n164)) {bUsePreviewValueAsDefault=true, inputName=Calcated_Bounds, sortPriority=11}) {name=Caclcated_Bounds, nodeColor=<rgba 0 0.9297 1 1>}
  n165 = TextureObjectParameter() {
   group=Wrap_EqualizerBounce
   parameterName=EquBounce__Tex__Normal
   samplerType=SAMPLERTYPE_Normal
   texture=&FlatNormalBlue
  }
  NamedRerouteDeclaration(None=@n165) {name=Tex_Face_Normal, nodeColor=<rgba 0.4531 0 1 1>}
  n166 = PreSkinnedNormal()
  NamedRerouteDeclaration(None=FunctionInput(Preview=VertexInterpolator(VS=@n166).PS) {bUsePreviewValueAsDefault=true, inputName=Pre-Skinned Local Normal, sortPriority=10}) {name=PreSkin_Normal, nodeColor=<rgba 0.8281 1 0 1>}
  NamedRerouteDeclaration(None=@n27.Normal) {name=Root_Normal, nodeColor=<rgba 0.4062 0 1 1>}
  NamedRerouteUsage()
  Power(Base=@n6)
  Saturate(None=@n46)
  ScalarParameter() {defaultValue=1.5, group=Wrap_EqualizerBounce_UVs, parameterName=EquBounce_Equlizer_Scale, sortPriority=150}
  ScalarParameter() {parameterName=daf}
  TextureSample(UVs=@n11.UVs, Tex=@n13) {desc=" ", samplerType=SAMPLERTYPE_Masks}
  VectorParameter() {group=Wrap_EqualizerBounce_UVs, parameterName=EquBounce_Equlizer_Offset, defaultValue=<rgba 0.27 0 0 1>}
 ]}
