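# Material function for the Finders Keepers cosmetic wrap, toggled by the Wrap_FKeepers static switch.
# Projects base color and two mask textures over pre-skinned bounds, with separate transforms for weapons, vehicles, side vehicles and mechs.
# A time-scrolled gradient, warped by mask channels, animates X, circle and line markers tinted over the base color.
# Two customization masks blend this animated look and a flat secondary finish into the incoming material attributes.
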MaterialFunction /Game/Weapons/WeaponSkins/Wraps/MaterialFunctions/MF_Wrap_FKeepers
graph {nodes [
  n1 = Time()
  n2 = ScalarParameter() {defaultValue=0.125, group=Wrap_FindersKeepers, parameterName=FK_Gradient_Speed}
  n3 = VertexInterpolator(VS=PreSkinnedPosition())
  n4 = PreSkinnedLocalBounds()
  n5 = FunctionInput(Preview=Divide(A=Subtract(A=@n3.PS, B=@n4.Min), B=Subtract(A=@n4.Max, B=@n4.Max))) {bUsePreviewValueAsDefault=true, inputName=Calcated_Bounds, sortPriority=11}
  n6 = FunctionInput(Preview=VertexInterpolator(VS=PreSkinnedNormal()).PS) {bUsePreviewValueAsDefault=true, inputName=Pre-Skinned Local Normal, sortPriority=10}
  n7 = Constant()
  n8 = MaterialFunctionCall(
   R=FunctionInput(Preview=Constant()) {inputName=isVehicle, inputType=FunctionInput_Scalar, sortPriority=5}
   G=FunctionInput(Preview=@n7) {bUsePreviewValueAsDefault=true, inputName=Side Project Vehicle, inputType=FunctionInput_Scalar}
   B=FunctionInput(Preview=@n7) {bUsePreviewValueAsDefault=true, inputName=isMech, inputType=FunctionInput_Scalar, sortPriority=4}
  ) {materialFunction=&AppendMany}
  n9 = MaterialFunctionCall(Color=@n8.RGB) {materialFunction=&SplitComponents}
  n10 = Multiply(
   A=Add(
    A=Multiply(A=@n1, B=@n2)
    B=MaterialFunctionCall!(
     For_Safe_Zones
     Normal_Map
     Caculated_Bounds=@n5
     Pre_Skinned_Normals=@n6
     Is Vehicle=@n9.R
     Is VehicleSide=@n9.R
     Is Mech=@n9.R
     Vehicle_Rotation
     UV_Rotation Bypass
     Mech_Rotation
     Uvs_Offset
     Offset-after rotation
     Uvs_TileRate
    ) {materialFunction=&/MF_Projection_Side_Normal}.UVs No Offsets/Tiling
   )
   B=ScalarParameter() {defaultValue=1, group=Wrap_FindersKeepers, parameterName=FK_Gradient_TileRate}
  )
  n11 = TextureObjectParameter() {group=Wrap__FindersKeepers_Textures, parameterName=FK_Mask_01, sortPriority=20, texture=&T_BlacksRGB}
  n12 = MaterialFunctionCall(Color=@n8.RGB) {materialFunction=&SplitComponents}
  n13 = LinearInterpolate(
   A=ScalarParameter() {defaultValue=1, group=Wrap_FindersKeepers, parameterName=FK_Blend_Contrast, sortPriority=34}
   B=ScalarParameter() {defaultValue=2, group=Wrap_FindersKeepers, parameterName=FK_Blend_Contrast_Mech, sortPriority=34}
   Alpha=@n12.B
  )
  n14 = ScalarParameter() {defaultValue=1, group=Wrap_FindersKeepers, parameterName=FK_Fill_alternate_Axis_Opacity 01}
  n15 = VectorParameter() {group=Wrap_FindersKeepers, parameterName=FK_Fill_alternate_Axis_Color_Mask_01, defaultValue=<rgba 0 0 0 1>}
  n16 = AppendVector(A=@n15.RGB, B=@n15.RGB)
  n17 = VectorParameter() {group=Wrap_FindersKeepers, parameterName=FK_Fill_Z_Axis_Weapon_Color_Mask_01, defaultValue=<rgba 0 0 0 1>}
  n18 = VectorParameter() {group=Wrap_FindersKeepers_UVS, parameterName=FK_Offset_Weapon_01---------------------, sortPriority=10, defaultValue=<rgba 0 0 0 1>}
  n19 = VectorParameter() {group=Wrap_FindersKeepers_UVS, parameterName=FK_Offset_Vehicle_01, sortPriority=11, defaultValue=<rgba 0 0 0 1>}
  n20 = VectorParameter() {group=Wrap_FindersKeepers_UVS, parameterName=FK_Offset_Vehicle_Side_01, sortPriority=11, defaultValue=<rgba 0 0 0 1>}
  n21 = LinearInterpolate(A=@n18.RGB, B=LinearInterpolate(A=@n19.RGB, B=@n20.RGB, Alpha=@n12.G), Alpha=@n12.R)
  n22 = VectorParameter() {group=Wrap_FindersKeepers_UVS, parameterName=FK_Offset_Mech_01, sortPriority=12, defaultValue=<rgba 0 0 0 1>}
  n23 = VectorParameter() {group=Wrap_FindersKeepers_UVS, parameterName="FK_Scale_Weapon_01 A=uniform", sortPriority=30, defaultValue=<rgba 1 1 1 1>}
  n24 = Multiply(A=@n23.RGB, B=@n23.RGB)
  n25 = VectorParameter() {group=Wrap_FindersKeepers_UVS, parameterName="FK_Scale_Vehicle_01 A=uniform", sortPriority=31, defaultValue=<rgba 1 1 1 1>}
  n26 = VectorParameter() {group=Wrap_FindersKeepers_UVS, parameterName="FK_Scale_Vehicle_Side_01 A=uniform", sortPriority=31, defaultValue=<rgba 1 1 1 1>}
  n27 = LinearInterpolate(A=Multiply(A=@n25.RGB, B=@n25.RGB), B=Multiply(A=@n26.RGB, B=@n26.RGB), Alpha=@n12.G)
  n28 = VectorParameter() {group=Wrap_FindersKeepers_UVS, parameterName="FK_Scale_Mech_01 A=uniform", defaultValue=<rgba 1 1 1 1>}
  n29 = Multiply(
   A=Add(A=@n5, B=LinearInterpolate(A=@n21, B=@n22.RGB, Alpha=@n12.B))
   B=LinearInterpolate(A=LinearInterpolate(A=@n24, B=@n27, Alpha=@n12.R), B=Multiply(A=@n28.RGB, B=@n28.RGB), Alpha=@n12.B)
  )
  n30 = VectorParameter() {group=Wrap_FindersKeepers_UVS, parameterName=FK-Rotation-Weapon-01---------------------, sortPriority=1, defaultValue=<rgba 0 0 -0.25 1>}
  n31 = VectorParameter() {group=Wrap_FindersKeepers_UVS, parameterName=FK_Rotation_Vehicle_01, sortPriority=2, defaultValue=<rgba 0 0 0 1>}
  n32 = VectorParameter() {group=Wrap_FindersKeepers_UVS, parameterName=FK_Rotation_Vehicle_Side_01, sortPriority=2, defaultValue=<rgba 0 0 0 1>}
  n33 = VectorParameter() {group=Wrap_FindersKeepers_UVS, parameterName=FK_Rotation_Mech_01, sortPriority=3, defaultValue=<rgba 0.25 0.25 0 1>}
  n34 = LinearInterpolate(A=LinearInterpolate(A=@n30.RGB, B=LinearInterpolate(A=@n31.RGB, B=@n32.RGB, Alpha=@n12.G), Alpha=@n12.R), B=@n33.RGB, Alpha=@n12.B)
  n35 = MaterialFunctionCall!(
   IsCharacter
   IsStatsB
   Texture=@n11
   Rotation_Axis=@n34
   Caculated_Bounds=@n29
   Pre_Skinned_Normals=@n6
   Blend_Contrast=@n13
   Fill_alternate_Axis_Opacity=@n14
   "Fill_alternate_Axis_Color_A=Roughness"=@n16
   Is_Vehicle=@n12.R
   Is_Mech=@n12.B
   Enable_Fill_Z_Weapon=StaticBool() {value=true}
   Fill_Z_Weapon_opacity=ScalarParameter() {group=Wrap_FindersKeepers, parameterName=FK_Fill_Z_Axis_Weapon_Opacity 01}
   Fill_Z_Weapon_Color=AppendVector(A=@n17.RGB, B=@n17.RGB)
  ) {materialFunction=&/MF_TopsSideProjection_Base}
  n36 = ChannelMaskParameter(None=@n35.Combined) {group=Wrap_FindersKeepers, maskChannel=Green, parameterName=FK_X01_WarpMask_01, sortPriority=-26, defaultValue=<rgba 0 1 0 0>}
  n37 = Add(
   A=@n10
   B=Multiply(A=@n36.R, B=ScalarParameter() {defaultValue=0.25, group=Wrap_FindersKeepers, parameterName=FK_X01_WarpAmount, sortPriority=-26})
  )
  n38 = TextureObjectParameter() {group=Wrap__FindersKeepers_Textures, parameterName=FK_GradientMap, texture=&T_BlacksRGB}
  n39 = ChannelMaskParameter(None=@n35.Combined) {group=Wrap_FindersKeepers, parameterName=FK_X01_ChannelMask_01, sortPriority=-26}
  n40 = Multiply(
   A=ChannelMaskParameter(None=TextureSample(UVs=@n37, Tex=@n38).RGBA) {group=Wrap_FindersKeepers, parameterName=FK_Gradient_ChannelMask_01}.R
   B=@n39.R
  )
  n41 = Saturate(None=Multiply(A=@n40, B=Add(A=@n40, B=@n39.R)))
  n42 = ChannelMaskParameter(None=@n35.Combined) {group=Wrap_FindersKeepers, maskChannel=Alpha, parameterName=FK_X02_WarpMask_01, sortPriority=-26, defaultValue=<rgba 0 0 0 1>}
  n43 = Add(
   A=@n10
   B=Multiply(A=@n42.R, B=ScalarParameter() {defaultValue=0.25, group=Wrap_FindersKeepers, parameterName=FK_X02_WarpAmount, sortPriority=-26})
  )
  n44 = ChannelMaskParameter(None=@n35.Combined) {group=Wrap_FindersKeepers, maskChannel=Blue, parameterName=FK_X02_ChannelMask_01, sortPriority=-26, defaultValue=<rgba 0 0 1 0>}
  n45 = Multiply(
   A=ChannelMaskParameter(None=TextureSample(UVs=@n43, Tex=@n38).RGBA) {group=Wrap_FindersKeepers, parameterName=FK_Gradient_ChannelMask_01}.R
   B=@n44.R
  )
  n46 = Saturate(None=Multiply(A=@n45, B=Add(A=@n45, B=@n44.R)))
  Add(A=@n41, B=@n46)
  n47 = TextureObjectParameter() {
   group=Wrap__FindersKeepers_Textures
   parameterName=FK_BaseColor_01
   sortPriority=11
   texture=&T_BlacksRGB
  }
  n48 = VectorParameter() {group=Wrap_FindersKeepers, parameterName="FK_Fill_alternate_Axis_Color_A=Roughness 01", defaultValue=<rgba 0.4735 0.4508 0.4179 0.5>}
  n49 = MaterialFunctionCall(
   Texture=@n47
   Rotation_Axis=@n34
   Caculated_Bounds=@n29
   Pre_Skinned_Normals=@n6
   Blend_Contrast=@n13
   Fill_alternate_Axis_Opacity=@n14
   "Fill_alternate_Axis_Color_A=Roughness"=AppendVector(A=@n48.RGB, B=@n48.RGB)
   Is_Vehicle=@n12.R
   Is_Mech=@n12.B
  ) {materialFunction=&MF_TopsSideProjection_Base}
  n50 = ComponentMask(None=@n49.Combined) {b=true, g=true, r=true}
  n51 = VectorParameter() {group=Wrap_FindersKeepers, parameterName=FK_Marker_BaseColor_01, sortPriority=2, defaultValue=<rgba 0.1301 0.0027 0.0052 1>}
  n52 = TextureObjectParameter() {group=Wrap__FindersKeepers_Textures, parameterName=FK_Mask_02, sortPriority=20, texture=&T_BlacksRGB}
  n53 = MaterialFunctionCall(
   Texture=@n52
   Rotation_Axis=@n34
   Caculated_Bounds=@n29
   Pre_Skinned_Normals=@n6
   Blend_Contrast=@n13
   Fill_alternate_Axis_Opacity=@n14
   "Fill_alternate_Axis_Color_A=Roughness"=@n16
   Is_Vehicle=@n12.R
   Is_Mech=@n12.B
  ) {materialFunction=&MF_TopsSideProjection_Base}
  n54 = ChannelMaskParameter(None=@n53.Combined) {group=Wrap_FindersKeepers, maskChannel=Green, parameterName=FK_Circles_WarpMask_01, sortPriority=-26, defaultValue=<rgba 0 1 0 0>}
  n55 = Add(
   A=@n10
   B=Multiply(A=@n54.R, B=ScalarParameter() {defaultValue=1, group=Wrap_FindersKeepers, parameterName=FK_Circle_WarpAmount, sortPriority=-26})
  )
  n56 = ChannelMaskParameter(None=@n53.Combined) {group=Wrap_FindersKeepers, parameterName=FK_Circles_ChannelMask_01, sortPriority=-26}
  n57 = Multiply(
   A=ChannelMaskParameter(None=TextureSample(UVs=@n55, Tex=@n38).RGBA) {group=Wrap_FindersKeepers, parameterName=FK_Gradient_ChannelMask_01}.R
   B=@n56.R
  )
  n58 = Multiply(A=@n57, B=Add(A=@n57, B=@n56.R))
  n59 = ChannelMaskParameter(None=@n53.Combined) {group=Wrap_FindersKeepers, maskChannel=Alpha, parameterName=FK_Lines_WarpMask_01, sortPriority=-26, defaultValue=<rgba 0 0 0 1>}
  n60 = Add(
   A=@n10
   B=Multiply(A=@n59.R, B=ScalarParameter() {defaultValue=1, group=Wrap_FindersKeepers, parameterName=FK_Line_WarpAmount, sortPriority=-26})
  )
  n61 = ChannelMaskParameter(None=@n53.Combined) {group=Wrap_FindersKeepers, maskChannel=Blue, parameterName=FK_Lines_ChannelMask_01, sortPriority=-26, defaultValue=<rgba 0 0 1 0>}
  n62 = Multiply(
   A=ChannelMaskParameter(None=TextureSample(UVs=@n60, Tex=@n38).RGBA) {group=Wrap_FindersKeepers, parameterName=FK_Gradient_ChannelMask_01}.R
   B=@n61.R
  )
  n63 = Add(A=@n62, B=@n61.R)
  n64 = Add(A=Add(A=@n41, B=Saturate(None=@n58)), B=Saturate(None=Multiply(A=@n62, B=@n63)))
  n65 = VectorParameter() {group=Wrap_FindersKeepers, parameterName=FK_Marker_BaseColor_02, sortPriority=2, defaultValue=<rgba 0.1301 0.0027 0.0052 0.5>}
  n66 = ChannelMaskParameter(None=@n49.Combined) {group=Wrap_FindersKeepers, maskChannel=Alpha, parameterName=FK_Rough_ChannelMask-----, sortPriority=-25, defaultValue=<rgba 0 0 0 1>}
  n67 = GetMaterialAttributes(None=FunctionInput() {inputName=Material, inputType=FunctionInput_MaterialAttributes, sortPriority=-20}) {attributeGetTypes [#1]}
  n68 = SetMaterialAttributes!(
   MaterialAttributes=@n67.MaterialAttributes
   BaseColor=LinearInterpolate(A=LinearInterpolate(A=@n50, B=@n51.RGB, Alpha=@n64), B=@n65.RGB, Alpha=@n46)
   Metallic=ScalarParameter() {group=Wrap_FindersKeepers, parameterName=FK_Metallic, sortPriority=1}
   Specular=ScalarParameter() {defaultValue=1, group=Wrap_FindersKeepers, parameterName=FK_Specular, sortPriority=1}
   Roughness=LinearInterpolate(
    A=ScalarParameter() {group=Wrap_FindersKeepers, parameterName=FK_Roughness, sortPriority=0}
    B=ScalarParameter() {defaultValue=1, group=Wrap_FindersKeepers, parameterName=FK_Roughness_High, sortPriority=0}
    Alpha=@n66.R
   )
  ) {attributeSetTypes [
    #2
    #3
    #4
    #5
   ]}
  n69 = FunctionInput() {
   bUsePreviewValueAsDefault=true
   desc=secondary
   inputName="Customization Mask 2 (_CM B Top Right)"
   inputType=FunctionInput_Scalar
   sortPriority=50
  }
  n70 = BlendMaterialAttributes(A=@n67.MaterialAttributes, B=@n68, Alpha=@n69) {vertexAttributeBlendType=UseA}
  n71 = VectorParameter() {group=Wrap_FindersKeepers_Secondary, parameterName=FK_Secondary_BaseColor, defaultValue=<rgba 0.1095 0.1221 0.0409 1>}
  n72 = SetMaterialAttributes!(
   MaterialAttributes=GetMaterialAttributes(None=@n70).MaterialAttributes
   BaseColor=@n71.RGB
   Metallic=ScalarParameter() {group=Wrap_FindersKeepers_Secondary, parameterName=FK_Secondary_Metallic}
   Specular=ScalarParameter() {defaultValue=0.55, group=Wrap_FindersKeepers_Secondary, parameterName=FK_Secondary_Specular}
   Roughness=ScalarParameter() {defaultValue=0.45, group=Wrap_FindersKeepers_Secondary, parameterName=FK_Secondary_Roughness}
   Normal=@n67.Normal
  ) {attributeSetTypes [
    #2
    #3
    #4
    #5
    #1
   ]}
  n73 = FunctionInput() {
   bUsePreviewValueAsDefault=true
   desc=base customization
   inputName="Customization Mask (_CM B Top left)"
   inputType=FunctionInput_Scalar
   sortPriority=49
  }
  FunctionOutput(None=StaticSwitchParameter(True=BlendMaterialAttributes(A=@n70, B=@n72, Alpha=@n73) {vertexAttributeBlendType=UseA}, False=@n67.MaterialAttributes) {group=Customization_Wrap, parameterName=Wrap_FKeepers})
  Reroute()
  Reroute()
 ]}
