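# MF_Blend4 is a material function in the Fortress shared base customization set that blends four pairs of inputs with one shared factor.
# Each A_0N and B_0N pair feeds its own linear interpolate, all driven by a single scalar Alpha.
# The results leave as Out_01 through Out_04, so four channels or properties crossfade together between an A set and a B set.
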MaterialFunction /Game/Packages/Fortress_SharedMaterials/Base_Material_Functions/BaseCustomizationFunctions/MF_Blend4
graph {nodes [
  n1 = FunctionInput() {inputName=Alpha, inputType=FunctionInput_Scalar, sortPriority=8}
  FunctionOutput(None=LinearInterpolate(A=FunctionInput() {inputName=A_01, sortPriority=-1}, B=FunctionInput() {inputName=B_01}, Alpha=@n1)) {outputName=Out_01}
  FunctionOutput(None=LinearInterpolate(A=FunctionInput() {inputName=A_02, sortPriority=1}, B=FunctionInput() {inputName=B_02, sortPriority=2}, Alpha=@n1)) {outputName=Out_02, sortPriority=2}
  FunctionOutput(None=LinearInterpolate(A=FunctionInput() {inputName=A_03, sortPriority=3}, B=FunctionInput() {inputName=B_03, sortPriority=4}, Alpha=@n1)) {outputName=Out_03, sortPriority=3}
  FunctionOutput(None=LinearInterpolate(A=FunctionInput() {inputName=A_04, sortPriority=4}, B=FunctionInput() {inputName=B_04, sortPriority=4}, Alpha=@n1)) {outputName=Out_04, sortPriority=4}
 ]}
