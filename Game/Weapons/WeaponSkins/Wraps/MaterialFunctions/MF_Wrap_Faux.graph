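# Material function that adds the Faux weapon wrap, a two-layer skin (female and male) divided by an animated zipper, behind the Wrap_Faux static switch.
# Side projection from pre-skinned normals and bounds maps each layer, with IsState channels selecting weapon, vehicle, vehicle-side or mech tuning.
# Blueprint-set start times drive the zipper slide and the female layer's emissive transition to a sine pulse.
# Customization masks blend the result over the input material.
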MaterialFunction /Game/Weapons/WeaponSkins/Wraps/MaterialFunctions/MF_Wrap_Faux
graph {nodes [
  Constant()
  Constant()
  Constant()
  Constant() {r=1}
  n1 = TextureSample() {samplerType=SAMPLERTYPE_Normal, texture=/Game/Weapons/FORT_Rifles/T_SCAR_N}
  n2 = GetMaterialAttributes(None=FunctionInput(Preview=SetMaterialAttributes!(MaterialAttributes, Normal=@n1.RGB) {attributeSetTypes [#1]}) {inputName=Material, inputType=FunctionInput_MaterialAttributes, sortPriority=-20}) {attributeGetTypes [#1]}
  n3 = TextureObjectParameter() {group=Wrap_Faux, parameterName=Faux_Female__Norm, samplerType=SAMPLERTYPE_Normal, texture=/Engine/EngineMaterials/FlatNormal}
  n4 = FunctionInput() {bUsePreviewValueAsDefault=true, inputName=IsState, inputType=FunctionInput_Scalar, sortPriority=4}
  n5 = MaterialFunctionCall(Color=@n4) {materialFunction=&SplitComponents}
  n6 = VectorParameter() {group=Wrap_Faux, parameterName=Faux_Uv_Offset_Vehicle, defaultValue=<rgba 0 0 0 1>}
  n7 = VectorParameter() {group=Wrap_Faux, parameterName=Faux_Uv_Offset_Vehicle_Side, defaultValue=<rgba 0 0 0 1>}
  n8 = MaterialFunctionCall(Color=@n4) {materialFunction=&SplitComponents}
  n9 = LinearInterpolate(
   A=LinearInterpolate(
    A=VectorParameter() {group=Wrap_Faux, parameterName=Faux_Uv_Offset_Weapon, defaultValue=<rgba 0 0 0 1>}.RGB
    B=LinearInterpolate(A=@n6.RGB, B=@n7.RGB, Alpha=@n8.G)
    Alpha=@n8.R
   )
   B=VectorParameter() {group=Wrap_Faux, parameterName=Faux_Uv_Offset_Mech, defaultValue=<rgba 0 0 0 1>}.RGB
   Alpha=@n8.B
  )
  n10 = ScalarParameter() {defaultValue=1, group=Wrap_Faux, parameterName=Faux_Uv_TileRate_Vehicle}
  n11 = ScalarParameter() {defaultValue=1, group=Wrap_Faux, parameterName=Faux_Uv_TileRate_Vehcile_Side}
  n12 = MaterialFunctionCall(Color=@n4) {materialFunction=&SplitComponents}
  n13 = LinearInterpolate(
   A=LinearInterpolate(
    A=ScalarParameter() {defaultValue=1, group=Wrap_Faux, parameterName=Faux_Uv_TileRate_Weapon}
    B=LinearInterpolate(A=@n10, B=@n11, Alpha=@n12.G)
    Alpha=@n12.R
   )
   B=ScalarParameter() {defaultValue=1, group=Wrap_Faux, parameterName=Faux_Uv_TileRate_Mech}
   Alpha=@n12.B
  )
  n14 = ScalarParameter() {group=Wrap_Faux, parameterName=Faux_Rotation_Vehicle, sortPriority=50}
  n15 = ScalarParameter() {group=Wrap_Faux, parameterName=Faux_Rotation_Vehcile_Side, sortPriority=50}
  n16 = MaterialFunctionCall(Color=@n4) {materialFunction=&SplitComponents}
  n17 = LinearInterpolate(
   A=LinearInterpolate(
    A=ScalarParameter() {group=Wrap_Faux, parameterName=Faux_Rotation_Weapon, sortPriority=50}
    B=LinearInterpolate(A=@n14, B=@n15, Alpha=@n16.G)
    Alpha=@n16.R
   )
   B=ScalarParameter() {defaultValue=-0.25, group=Wrap_Faux, parameterName=Faux_Rotation_Mech, sortPriority=50}
   Alpha=@n16.B
  )
  n18 = FunctionInput(Preview=VertexInterpolator(VS=PreSkinnedNormal()).PS) {bUsePreviewValueAsDefault=true, inputName=Pre-Skinned Local Normal, sortPriority=10}
  n19 = VertexInterpolator(VS=PreSkinnedPosition())
  n20 = PreSkinnedLocalBounds()
  n21 = FunctionInput(Preview=Divide(A=Subtract(A=@n19.PS, B=@n20.Min), B=Subtract(A=@n20.Max, B=@n20.Max))) {bUsePreviewValueAsDefault=true, inputName=Calcated_Bounds, sortPriority=11}
  n22 = MaterialFunctionCall!(
   For_Safe_Zones
   Normal_Map=@n3
   Caculated_Bounds=@n21
   Pre_Skinned_Normals=@n18
   Is Vehicle=@n5.R
   Is VehicleSide=@n5.R
   Is Mech=@n5.R
   Vehicle_Rotation
   UV_Rotation Bypass=@n17
   Mech_Rotation
   Uvs_Offset=@n9
   Offset-after rotation
   Uvs_TileRate=@n13
  ) {materialFunction=&/MF_Projection_Side_Normal}
  n23 = TextureSample(
   UVs=@n22.UVs
   Tex=TextureObjectParameter() {group=Wrap_Faux, parameterName=Faux_Female__BC, texture=&T_BlacksRGB}
  )
  n24 = VectorParameter() {group=Wrap_Faux, parameterName="Faux_Male_Tri_Color   A=Roughness", defaultValue=<rgba 0 0 0 1>}
  n25 = TextureObjectParameter() {group=Wrap_Faux, parameterName=Faux_Male__Norm, samplerType=SAMPLERTYPE_Normal, texture=&FlatNormal}
  n26 = MaterialFunctionCall(Color=@n4) {materialFunction=&SplitComponents}
  n27 = MaterialFunctionCall(
   Normal_Map=@n25
   Caculated_Bounds=@n21
   Pre_Skinned_Normals=@n18
   Is Vehicle=@n26.R
   Is VehicleSide=@n26.R
   Is Mech=@n26.R
   UV_Rotation Bypass=@n17
   Uvs_Offset=@n9
   Uvs_TileRate=@n13
  ) {materialFunction=&MF_Projection_Side_Normal}
  n28 = TextureSample(
   UVs=@n27.UVs
   Tex=TextureObjectParameter() {group=Wrap_Faux, parameterName=Faux_Male__BC, texture=&T_BlacksRGB}
  )
  n29 = TextureObjectParameter() {group=Wrap_Faux, parameterName=Faux_Zipper__Mask, texture=&T_BlacksRGB}
  n30 = MaterialFunctionCall(Color=Abs(None=@n18)) {materialFunction=&SplitComponents}
  n31 = MaterialFunctionCall(Color=@n4) {materialFunction=&SplitComponents}
  n32 = LinearInterpolate(A=@n30.B, B=@n30.B, Alpha=@n31.G)
  n33 = MaterialFunctionCall(Float4=@n4) {materialFunction=&BreakOutFloat4Components}
  n34 = MaterialFunctionCall(
   In=LinearInterpolate(A=LinearInterpolate(A=@n30.R, B=@n32, Alpha=@n31.R), B=@n30.G, Alpha=@n31.B)
   Contrast=LinearInterpolate(
    A=ScalarParameter() {defaultValue=5, group=Wrap_Faux, parameterName=Faux_Tri_Mask, sortPriority=10}
    B=ScalarParameter() {defaultValue=5, group=Wrap_Faux, parameterName=Faux_Tri_Mask_Vehicle, sortPriority=10}
    Alpha=@n33.R
   )
  ) {materialFunction=&CheapContrast}
  n35 = Saturate(None=Add(A=TextureSample(UVs=@n27.UVs, Tex=@n29).A, B=@n34.Result))
  n36 = TextureObjectParameter() {group=Wrap_Faux, parameterName=Faux_Zipper__Norm, samplerType=SAMPLERTYPE_Normal, texture=&FlatNormal}
  n37 = MaterialFunctionCall(Color=@n4) {materialFunction=&SplitComponents}
  n38 = MaterialFunctionCall(Color=@n4) {materialFunction=&SplitComponents}
  n39 = LinearInterpolate(
   A=VectorParameter() {group=Wrap_Faux, parameterName=Faux_Zipper_Offset_Weapon, defaultValue=<rgba 0 0 0 1>}.RGB
   B=LinearInterpolate(
    A=VectorParameter() {group=Wrap_Faux, parameterName=Faux_Zipper_Offset_Vehicle, defaultValue=<rgba 0 0 0 1>}.RGB
    B=VectorParameter() {group=Wrap_Faux, parameterName=Faux_Zipper_Offset_Vehicle_Side, defaultValue=<rgba 0 0 0 1>}.RGB
    Alpha=@n38.G
   )
   Alpha=@n38.R
  )
  n40 = VectorParameter() {group=Wrap_Faux, parameterName=Faux_Zipper_Offset_Mech, defaultValue=<rgba 0 0 0 1>}
  n41 = Divide(
   A=Subtract(A=Time(), B=ScalarParameter() {defaultValue=-9999999, group=Wrap_Faux, parameterName=Faux_TimeFrom_BP, sortPriority=100})
   B=ScalarParameter() {defaultValue=0.25, group=Wrap_Faux, parameterName=Faux_TimeFrom_BP_Duration, sortPriority=100}
  )
  n42 = LinearInterpolate(
   A=ScalarParameter() {defaultValue=1, group=Wrap_Faux, parameterName=Faux_Zipper_Position_Start, sortPriority=100}
   B=ScalarParameter() {defaultValue=1, group=Wrap_Faux, parameterName=Faux_Zipper_Position_End, sortPriority=100}
   Alpha=OneMinus(None=Saturate(None=@n41))
  )
  n43 = Constant2Vector()
  n44 = ScalarParameter() {defaultValue=1, group=Wrap_Faux, parameterName=Faux_Zipper_TileRate_Weapon}
  n45 = MaterialFunctionCall(Color=@n4) {materialFunction=&SplitComponents}
  n46 = LinearInterpolate(
   A=ScalarParameter() {defaultValue=1, group=Wrap_Faux, parameterName=Faux_Zipper_TileRate_Vehicle}
   B=ScalarParameter() {defaultValue=1, group=Wrap_Faux, parameterName=Faux_Zipper_TileRate_Vehcile_Side}
   Alpha=@n45.G
  )
  n47 = MaterialFunctionCall(
   Normal_Map=@n36
   Caculated_Bounds=@n21
   Pre_Skinned_Normals=@n18
   Is Vehicle=@n37.R
   Is VehicleSide=@n37.R
   Is Mech=@n37.R
   UV_Rotation Bypass=@n17
   Uvs_Offset=Add(A=LinearInterpolate(A=@n39, B=@n40.RGB, Alpha=@n38.B), B=AppendVector(A=@n42, B=@n43))
   Uvs_TileRate=LinearInterpolate(
    A=LinearInterpolate(A=@n44, B=@n46, Alpha=@n45.R)
    B=ScalarParameter() {defaultValue=1, group=Wrap_Faux, parameterName=Faux_Zipper_TileRate_Mech}
    Alpha=@n45.B
   )
  ) {materialFunction=&MF_Projection_Side_Normal}
  n48 = TextureObjectParameter() {group=Wrap_Faux, parameterName=Faux_Zipper__BC, texture=&T_BlacksRGB}
  n49 = TextureSample(UVs=@n47.UVs, Tex=@n48)
  n50 = LinearInterpolate(A=@n49.B, B=OneMinus(None=@n49.B), Alpha=ScalarParameter() {group=Wrap_Faux, parameterName=Faux_Zipper_Invert_Mask})
  n51 = TextureSample(UVs=@n22.UVs, Tex=@n29)
  n52 = ChannelMaskParameter(None=@n51.RGBA) {group=Wrap_Faux, maskChannel=Blue, parameterName=Faux_Female_Metalic_Channel, defaultValue=<rgba 0 0 1 0>}
  n53 = MaterialFunctionCall!(
   A_01=@n23.RGB
   B_01=LinearInterpolate(A=@n24.RGB, B=@n28.RGB, Alpha=@n35)
   A_02=LinearInterpolate(
    A=ScalarParameter() {group=Wrap_Faux, parameterName=Faux_Female_Roughness_Low}
    B=ScalarParameter() {defaultValue=1, group=Wrap_Faux, parameterName=Faux_Female_Roughness_High}
    Alpha=@n23.A
   )
   B_02=LinearInterpolate(
    A=ScalarParameter() {group=Wrap_Faux, parameterName=Faux_Male_Roughness_Low}
    B=ScalarParameter() {defaultValue=1, group=Wrap_Faux, parameterName=Faux_Male_Roughness_High}
    Alpha=LinearInterpolate(A=@n24.A, B=@n28.A, Alpha=@n35)
   )
   A_03=ScalarParameter() {defaultValue=0.5, group=Wrap_Faux, parameterName=Faux_Femail_Specular}
   B_03=ScalarParameter() {defaultValue=0.5, group=Wrap_Faux, parameterName=Faux_Femail_Male_Spec}
   A_04=@n52.R
   B_04=ScalarParameter() {group=Wrap_Faux, parameterName=Faux_Femail_Male_Metalic}
   Alpha=@n50
  ) {
   desc=Blending of materials
   materialFunction=/Game/Packages/Fortress_SharedMaterials/Base_Material_Functions/BaseCustomizationFunctions/MF_Blend4
  }
  n54 = TextureSample(UVs=@n47.UVs, Tex=@n29)
  n55 = ChannelMaskParameter(None=@n49.RGBA) {group=Wrap_Faux, maskChannel=Green, parameterName=Faux_Zipper__Metalic, defaultValue=<rgba 0 1 0 0>}
  n56 = TextureSample(UVs=@n47.UVs, Tex=@n48)
  n57 = Multiply(A=Multiply(A=@n34.Result, B=@n56.A), B=@n49.B)
  n58 = MaterialFunctionCall(
   A_01=Multiply(A=@n53.Out_01, B=OneMinus(None=@n54.G))
   B_01=@n56.R
   A_02=Add(A=@n53.Out_02, B=@n54.G)
   B_02=ScalarParameter() {defaultValue=0.3, group=Wrap_Faux, parameterName=Faux_Zipper__Roughness}
   A_03=@n53.Out_03
   B_03=ScalarParameter() {defaultValue=0.5, group=Wrap_Faux, parameterName=Faux_Zipper__Spec}
   A_04=@n53.Out_03
   B_04=@n55.R
   Alpha=@n57
  ) {
   desc=MAsking of the zipper
   materialFunction=&MF_Blend4
  }
  n59 = LinearInterpolate(A=@n22.Normal, B=@n27.Normal, Alpha=@n50)
  n60 = ScalarParameter() {defaultValue=0.25, group=Wrap_Faux, parameterName=Faux_Female_Emissive_End, sortPriority=130}
  n61 = Time()
  n62 = Sine(None=Multiply(A=@n61, B=ScalarParameter() {defaultValue=0.5, group=Wrap_Faux, parameterName=Faux_Female_Emissive_Sine_Speed}))
  n63 = Saturate(None=Multiply(A=Add(A=@n62) {constB=3}) {constB=0.25})
  n64 = Divide(
   A=Subtract(A=@n61, B=ScalarParameter() {defaultValue=-9999999, group=Wrap_Faux, parameterName=Faux_TimeFrom_BP_Emis, sortPriority=120})
   B=ScalarParameter() {defaultValue=0.25, group=Wrap_Faux, parameterName=Faux_TimeFrom_BP_Emis_Duration, sortPriority=120}
  )
  n65 = Multiply(
   A=Multiply(A=@n51.R, B=VectorParameter() {group=Wrap_Faux, parameterName=Faux_Female_Emissive_Color, defaultValue=<rgba 0.0782 0.7157 0.7991 1>}.RGB)
   B=LinearInterpolate(
    A=ScalarParameter() {defaultValue=1, group=Wrap_Faux, parameterName=Faux_Female_Emissive_Start, sortPriority=130}
    B=Multiply(A=@n60, B=@n63)
    Alpha=Saturate(None=@n64)
   )
  )
  n66 = Constant()
  n67 = SetMaterialAttributes!(
   MaterialAttributes=@n2.MaterialAttributes
   BaseColor=@n58.Out_01
   Roughness=@n58.Out_01
   Specular=@n58.Out_01
   Metallic=@n58.Out_01
   Normal=Add(A=@n2.Normal, B=LinearInterpolate(A=@n59, B=@n47.Normal, Alpha=@n57))
   EmissiveColor=LinearInterpolate(A=LinearInterpolate(A=@n65, B=@n66, Alpha=@n50), B=@n66, Alpha=@n57)
  ) {attributeSetTypes [
    #2
    #3
    #4
    #5
    #1
    #6
   ]}
  n68 = FunctionInput() {bUsePreviewValueAsDefault=true, inputName="Customization Mask (_CM B)", inputType=FunctionInput_Scalar, sortPriority=24}
  n69 = NamedRerouteUsage()
  n70 = NamedRerouteUsage()
  n71 = LinearInterpolate(
   A=@n68
   B=FunctionInput() {bUsePreviewValueAsDefault=true, inputName="Customization Mask (_CM B)B", inputType=FunctionInput_Scalar, sortPriority=24}
   Alpha=Saturate(None=Subtract(A=@n69, B=@n70))
  )
  FunctionOutput(None=StaticSwitchParameter(True=BlendMaterialAttributes(A=@n2.MaterialAttributes, B=@n67, Alpha=@n71) {vertexAttributeBlendType=UseA}, False=@n2.MaterialAttributes) {group=Customization_Wrap, parameterName=Wrap_Faux}) {outputName=Result_1}
  Multiply(A=MaterialFunctionCall!(X range Y range, Gradient Height, ShowGrid?, Thickness, Color, "f(x)"=@n63) {materialFunction=/Engine/Functions/Engine_MaterialFunctions01/Debug/PlotFunctionOnGraph}."f(x) = y")
  NamedRerouteDeclaration(None=@n33.A) {name=IsCharacter, nodeColor=<rgba 0.8281 1 0 1>}
  NamedRerouteDeclaration(None=FunctionInput() {bUsePreviewValueAsDefault=true, inputName=isCharAccent, inputType=FunctionInput_Scalar, sortPriority=78}) {name=isCharAccent, nodeColor=<rgba 0 0.9253 1 1>}
  Saturate(None=Add(A=@n68, B=ScalarParameter() {group=Wrap_Faux, parameterName=Faux_Remove_Customization Mask, sortPriority=-10}))
 ]}
