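# Material function for the Flapjack Wrangler weapon wrap, blending an animated syrup look into incoming material attributes via customization masks and a static switch.
# Side-projected UVs, tuned by IsState for weapon, vehicle and mech, drive background and syrup textures.
# HeightLerp merges syrup shape with scrolling drips, then curve atlases remap it to color and perturbed normals.
# Syrup emissive adds a sphere-mask inner glow, scaled by time of day.
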MaterialFunction /Game/Weapons/WeaponSkins/Wraps/MaterialFunctions/MF_Wrap_FlapjackWrangler
graph {nodes [
  Constant()
  n1 = NamedRerouteUsage()
  n2 = FunctionInput(Preview=Constant()) {bUsePreviewValueAsDefault=true, inputName=IsState, inputType=FunctionInput_Vector4}
  n3 = MaterialFunctionCall(Color=@n2) {materialFunction=&SplitComponents}
  n4 = LinearInterpolate(
   A=ScalarParameter() {defaultValue=1, group=Wrap_FlapjackWrang, parameterName=FJW_BackGround_Scale_Weapon}
   B=LinearInterpolate(
    A=ScalarParameter() {defaultValue=1, group=Wrap_FlapjackWrang, parameterName=FJW_BackGround_Scale_Vehicle}
    B=ScalarParameter() {defaultValue=1, group=Wrap_FlapjackWrang, parameterName=FJW_BackGround_Scale_VehicleSide}
    Alpha=@n3.G
   )
   Alpha=@n3.R
  )
  n5 = ScalarParameter() {defaultValue=1, group=Wrap_FlapjackWrang, parameterName=FJW_BackGround_Scale_Mech}
  n6 = VectorParameter() {group=Wrap_FlapjackWrang, parameterName=FJW_Background_Offset_Weapon, defaultValue=<rgba 0 0 0 1>}
  n7 = MaterialFunctionCall(Color=@n2) {materialFunction=&SplitComponents}
  n8 = LinearInterpolate(
   A=VectorParameter() {group=Wrap_FlapjackWrang, parameterName=FJW_Background_Offset_Vehicle, defaultValue=<rgba 0 0 0 1>}.RGB
   B=VectorParameter() {group=Wrap_FlapjackWrang, parameterName=FJW_Background_Offset_VehicleSide, defaultValue=<rgba 0 0 0 1>}.RGB
   Alpha=@n7.G
  )
  n9 = ComponentMask(None=LinearInterpolate(
    A=LinearInterpolate(A=@n6.RGB, B=@n8, Alpha=@n7.R)
    B=VectorParameter() {group=Wrap_FlapjackWrang, parameterName=FJW_Background_Offset_Mech, defaultValue=<rgba 0 0 0 1>}.RGB
    Alpha=@n7.B
   )) {g=true, r=true}
  n10 = PreSkinnedLocalBounds()
  n11 = Subtract(A=VertexInterpolator(VS=PreSkinnedPosition()).PS, B=@n10.Min)
  n12 = Subtract(A=@n10.Max, B=@n10.Max)
  n13 = MaterialFunctionCall(Color=@n2) {materialFunction=&SplitComponents}
  n14 = MaterialFunctionCall!(
   For_Safe_Zones
   Normal_Map
   Caculated_Bounds=FunctionInput(Preview=Divide(A=@n11, B=@n12)) {bUsePreviewValueAsDefault=true, inputName=Calcated_Bounds, sortPriority=11}
   Pre_Skinned_Normals
   Is Vehicle=@n13.R
   Is VehicleSide=@n13.R
   Is Mech=@n13.R
   Vehicle_Rotation
   UV_Rotation Bypass
   Mech_Rotation
   Uvs_Offset
   Offset-after rotation
   Uvs_TileRate
  ) {materialFunction=&/MF_Projection_Side_Normal}
  n15 = ScalarParameter() {group=Wrap_FlapjackWrang, parameterName=FJW_BackGround_Rotation_Weapon}
  n16 = MaterialFunctionCall(Color=@n2) {materialFunction=&SplitComponents}
  n17 = LinearInterpolate(
   A=ScalarParameter() {group=Wrap_FlapjackWrang, parameterName=FJW_BackGround_Rotation_Vehicle}
   B=ScalarParameter() {group=Wrap_FlapjackWrang, parameterName=FJW_BackGround_Rotation_VehicleSide}
   Alpha=@n16.G
  )
  n18 = MaterialFunctionCall(
   UVs=@n14.UVs No Offsets/Tiling
   "Rotation Angle (0-1)"=LinearInterpolate(
    A=LinearInterpolate(A=@n15, B=@n17, Alpha=@n16.R)
    B=ScalarParameter() {group=Wrap_FlapjackWrang, parameterName=FJW_BackGround_Rotation_Mech}
    Alpha=@n16.B
   )
  ) {materialFunction=&CustomRotator}
  n19 = TextureSample(
   UVs=Multiply(A=LinearInterpolate(A=@n4, B=@n5, Alpha=@n3.B), B=Add(A=@n9, B=@n18.Rotated Values))
   Tex=TextureObjectParameter() {group=Wrap_FlapjackWrang, parameterName="FJW_BackGround__Tex  ", texture=&T_BlacksRGB}
  )
  n20 = NamedRerouteUsage()
  n21 = NamedRerouteUsage()
  n22 = ScalarParameter() {defaultValue=0.5, group=Wrap_FlapjackWrang, parameterName=FJW_Metalic, sortPriority=0}
  n23 = NamedRerouteUsage()
  n24 = NamedRerouteUsage()
  n25 = ScalarParameter() {defaultValue=0.5, group=Wrap_FlapjackWrang, parameterName=FJW_Specular, sortPriority=0}
  n26 = NamedRerouteUsage()
  n27 = ScalarParameter() {defaultValue=0.5, group=Wrap_FlapjackWrang, parameterName=FJW_Specular_Syrup, sortPriority=0}
  n28 = VectorParameter() {group=Wrap_FlapjackWrang_Uvs, parameterName=FJW_Syrup_Offset_Weapon, defaultValue=<rgba 0 1 0 1>}
  n29 = MaterialFunctionCall(Color=NamedRerouteUsage()) {materialFunction=&SplitComponents}
  n30 = LinearInterpolate(
   A=VectorParameter() {group=Wrap_FlapjackWrang_Uvs, parameterName=FJW_Syrup_Offset_Vehicle, defaultValue=<rgba 0 0 0 1>}.RGB
   B=VectorParameter() {group=Wrap_FlapjackWrang_Uvs, parameterName=FJW_Syrup_Offset_VehicleSide, defaultValue=<rgba 0 1 0 1>}.RGB
   Alpha=@n29.G
  )
  n31 = ComponentMask(None=LinearInterpolate(
    A=LinearInterpolate(A=@n28.RGB, B=@n30, Alpha=@n29.R)
    B=VectorParameter() {group=Wrap_FlapjackWrang_Uvs, parameterName=FJW_Syrup_Offset_Mech, defaultValue=<rgba 0 1 0 1>}.RGB
    Alpha=@n29.B
   )) {g=true, r=true}
  n32 = NamedRerouteUsage()
  n33 = ScalarParameter() {group=Wrap_FlapjackWrang_Uvs, parameterName=FJW_Syrup_Rotation_Vehicle}
  n34 = ScalarParameter() {group=Wrap_FlapjackWrang_Uvs, parameterName=FJW_Syrup_Rotation_VehicleSide}
  n35 = MaterialFunctionCall(Color=NamedRerouteUsage()) {materialFunction=&SplitComponents}
  n36 = LinearInterpolate(
   A=LinearInterpolate(
    A=ScalarParameter() {group=Wrap_FlapjackWrang_Uvs, parameterName=FJW_Syrup_Rotation_Weapon}
    B=LinearInterpolate(A=@n33, B=@n34, Alpha=@n35.G)
    Alpha=@n35.R
   )
   B=ScalarParameter() {group=Wrap_FlapjackWrang_Uvs, parameterName=FJW_Syrup_Rotation_Mech}
   Alpha=@n35.B
  )
  n37 = ScalarParameter() {defaultValue=1, group=Wrap_FlapjackWrang_Uvs, parameterName=FJW_Syrup_Scale_Vehicle}
  n38 = ScalarParameter() {defaultValue=1, group=Wrap_FlapjackWrang_Uvs, parameterName=FJW_Syrup_Scale_VehicleSide}
  n39 = MaterialFunctionCall(Color=NamedRerouteUsage()) {materialFunction=&SplitComponents}
  n40 = LinearInterpolate(
   A=LinearInterpolate(
    A=ScalarParameter() {defaultValue=1, group=Wrap_FlapjackWrang_Uvs, parameterName=FJW_Syrup_Scale_Weapon}
    B=LinearInterpolate(A=@n37, B=@n38, Alpha=@n39.G)
    Alpha=@n39.R
   )
   B=ScalarParameter() {defaultValue=1, group=Wrap_FlapjackWrang_Uvs, parameterName=FJW_Syrup_Scale_Mech}
   Alpha=@n39.B
  )
  n41 = Multiply(
   A=@n40
   B=Add(
    A=@n31
    B=MaterialFunctionCall(UVs=@n32, "Rotation Angle (0-1)"=@n36) {materialFunction=&CustomRotator}.Rotated Values
   )
  )
  n42 = Add(A=@n41, B=NamedRerouteUsage())
  n43 = TextureObjectParameter() {group=Wrap_FlapjackWrang, parameterName="FJW_Syrup__Tex  ", texture=&T_BlacksRGB}
  n44 = ChannelMaskParameter(None=TextureSample(UVs=Clamp(None=@n42) {maxDefault=0.994, minDefault=0.006}, Tex=@n43).RGBA) {group=Wrap_FlapjackWrang_Syurp, maskChannel=Alpha, parameterName=FJW_Syrup__Tex_maskChannel, defaultValue=<rgba 0 0 0 1>}
  n45 = Constant() {r=0.5}
  n46 = MaterialFunctionCall(Float2=@n41) {materialFunction=&BreakOutFloat2Components}
  n47 = ScalarParameter() {defaultValue=0.5, group=Wrap_FlapjackWrang_Syurp, parameterName=FJW_Syrup_Clamp Size}
  n48 = Add(
   A=Multiply(A=SphereMask(A=@n45, B=@n46.R, Radius=@n47) {attenuationRadius=0.5}, B=SphereMask(A=@n46.G, B=@n45, Radius=@n47) {attenuationRadius=0.5})
   B=ScalarParameter() {group=Wrap_FlapjackWrang_Syurp, parameterName="FJW_Syrup_Clamp_Mask 0=on"}
  )
  n49 = VertexInterpolator(VS=PreSkinnedNormal())
  n50 = MaterialFunctionCall(Color=Abs(None=FunctionInput(Preview=@n49.PS) {bUsePreviewValueAsDefault=true, inputName=Pre-Skinned Local Normal, sortPriority=10})) {materialFunction=&SplitComponents}
  n51 = MaterialFunctionCall(Color=@n2) {materialFunction=&SplitComponents}
  n52 = LinearInterpolate(A=LinearInterpolate(A=@n50.R, B=LinearInterpolate(A=@n50.B, B=@n50.B, Alpha=@n51.G), Alpha=@n51.R), B=@n50.G, Alpha=@n51.B)
  n53 = ScalarParameter() {group=Wrap_FlapjackWrang, parameterName=FJW_SyrupDrip__TriplannerMask_Contrast}
  n54 = VectorParameter() {group=Wrap_FlapjackWrang_Uvs, parameterName=FJW_SyrupDrip_Speed_Vehicle, defaultValue=<rgba 0 0 0 1>}
  n55 = VectorParameter() {group=Wrap_FlapjackWrang_Uvs, parameterName=FJW_SyrupDrip_Speed_VehicleSide, defaultValue=<rgba 0 0 0 1>}
  n56 = MaterialFunctionCall(Color=NamedRerouteUsage()) {materialFunction=&SplitComponents}
  n57 = LinearInterpolate(
   A=LinearInterpolate(
    A=VectorParameter() {group=Wrap_FlapjackWrang_Uvs, parameterName=FJW_SyrupDrip_Speed_Weapon, defaultValue=<rgba 0 1 0 1>}.RGB
    B=LinearInterpolate(A=@n54.RGB, B=@n55.RGB, Alpha=@n56.G)
    Alpha=@n56.R
   )
   B=VectorParameter() {group=Wrap_FlapjackWrang_Uvs, parameterName=FJW_SyrupDrip_Speed_Mech, defaultValue=<rgba 0 0 0 1>}.RGB
   Alpha=@n56.B
  )
  n58 = Time()
  n59 = MaterialFunctionCall(Color=NamedRerouteUsage()) {materialFunction=&SplitComponents}
  n60 = LinearInterpolate(
   A=ScalarParameter() {group=Wrap_FlapjackWrang_Uvs, parameterName=FJW_SyrupDrip_Rotation_Weapon}
   B=LinearInterpolate(
    A=ScalarParameter() {group=Wrap_FlapjackWrang_Uvs, parameterName=FJW_SyrupDrip_Rotation_Vehicle}
    B=ScalarParameter() {group=Wrap_FlapjackWrang_Uvs, parameterName=FJW_SyrupDrip_Rotation_VehicleSide}
    Alpha=@n59.G
   )
   Alpha=@n59.R
  )
  n61 = ScalarParameter() {group=Wrap_FlapjackWrang_Uvs, parameterName=FJW_SyrupDrip_Rotation_Mech}
  n62 = Add(
   A=ComponentMask(None=Multiply(A=@n57, B=@n58)) {g=true, r=true}
   B=MaterialFunctionCall(UVs=NamedRerouteUsage(), "Rotation Angle (0-1)"=LinearInterpolate(A=@n60, B=@n61, Alpha=@n59.B)) {materialFunction=&CustomRotator}.Rotated Values
  )
  n63 = Multiply(A=Multiply(A=Add(A=@n62, B=@n31), B=@n40), B=NamedRerouteUsage())
  n64 = NamedRerouteUsage()
  n65 = ChannelMaskParameter(None=TextureSample(
    UVs=Add(A=@n63, B=@n64)
    Tex=TextureObjectParameter() {group=Wrap_FlapjackWrang, parameterName="FJW_SyrupDrip__Tex  ", texture=&T_BlacksRGB}
   ).RGBA) {group=Wrap_FlapjackWrang, parameterName=FJW_SyrupDrip__Tex_maskChannel}
  n66 = VectorParameter() {
   group=Wrap_FlapjackWrang_Syurp
   parameterName=FJW_SyrupDryp_Mask_Radus
   sortPriority=1
   channelNames {a=hardness, b="", g=Y, r=X}
   defaultValue=<rgba 0.5 0.5 0 100>
  }
  n67 = SphereMask(A=@n45, B=@n46.R, Radius=@n66.X, Hardness=@n66.X) {attenuationRadius=0.5}
  n68 = MaterialFunctionCall(
   Transition Phase=LinearInterpolate(
    A=ScalarParameter() {defaultValue=0.6, group=Wrap_FlapjackWrang_Syurp, parameterName=FJW_Syrup_Clamp_non Value}
    B=@n44.R
    Alpha=Saturate(None=@n48)
   )
   Height Texture=Multiply(
    A=MaterialFunctionCall(In=@n52, Contrast=@n53) {materialFunction=&CheapContrast}.Result
    B=Multiply(A=@n65.R, B=@n67)
   )
   Contrast=ScalarParameter() {group=Wrap_FlapjackWrang_Syurp, parameterName=FJW_Syrup_HeightLerp_Contrast}
  ) {materialFunction=&HeightLerp}
  n69 = CurveAtlasRowParameter(CurveTime=@n68.Alpha) {
   atlas=/Game/Weapons/WeaponSkins/Wraps/Textures/CruveAtlas_Wraps_01
   curve=/Game/Weapons/WeaponSkins/Wraps/Textures/FlapjackWrangler/CC_Wrap_FlapjackWrangler01
   defaultValue=36
   group=Wrap_FlapjackWrang
   parameterName=FJW_Syrup_RemaptoColorCurve ForNormal
  }
  MaterialExpressionMaterialFunctionCall_36 = MaterialFunctionCall!(
   Index=ScalarParameter() {defaultValue=36, group=Wrap_FlapjackWrang, parameterName="FJW_Syrup_RemaptoColorCurve ForNormal  Index"}
   CCinput=@n69
   CCinput_Alpha=@n69
   UV=@n68.Alpha
  ) {materialFunction=&/MF_Global_CC_Tex}
  n70 = Convert(None=@MaterialExpressionMaterialFunctionCall_36.4v) {
   nodeName=Break Float4
   convertInputs [{
     type=Vector4
     defaultValue=<rgba 0 0 0 1>
     expressionInput {
      expression=&self:MaterialExpressionMaterialFunctionCall_36
      inputName=None
      mask=0
      maskA=0
      maskB=0
      maskG=0
      maskR=0
      outputIndex=1
     }
    }]
   convertOutputs [{type=Scalar}, {type=Scalar}, {type=Scalar}, {type=Scalar}]
  }
  n71 = Saturate(None=Power(Base=@n70, Exp=ScalarParameter() {defaultValue=2, group=Wrap_FlapjackWrang, parameterName=FJW_Specular_Syrup_power, sortPriority=0}))
  n72 = NamedRerouteUsage()
  n73 = LinearInterpolate(
   A=ScalarParameter() {group=Wrap_FlapjackWrang, parameterName=FJW_Roughness_Low, sortPriority=0}
   B=ScalarParameter() {defaultValue=1, group=Wrap_FlapjackWrang, parameterName=FJW_Roughness_High, sortPriority=0}
   Alpha=@n19.A
  )
  n74 = NamedRerouteUsage()
  n75 = LinearInterpolate(A=NamedRerouteUsage(), Alpha=NamedRerouteUsage()) {constB=0}
  n76 = NamedRerouteUsage()
  n77 = CurveAtlasRowParameter(CurveTime=@n68.Alpha) {
   atlas=&CruveAtlas_Wraps_01
   curve=/Game/Weapons/WeaponSkins/Wraps/Textures/FlapjackWrangler/CC_Wrap_FlapjackWrangler
   defaultValue=35
   group=Wrap_FlapjackWrang
   parameterName="FJW_Syrup_RemaptoColorCurve  A=Mask"
  }
  n78 = MaterialFunctionCall(
   Index=ScalarParameter() {defaultValue=35, group=Wrap_FlapjackWrang, parameterName="FJW_Syrup_RemaptoColorCurve  A=Mask  Index"}
   CCinput=@n77
   CCinput_Alpha=@n77
   UV=@n68.Alpha
  ) {materialFunction=&MF_Global_CC_Tex}
  n79 = ComponentMask(None=NamedRerouteUsage()) {g=true, r=true}
  n80 = ScalarParameter() {group=Wrap_FlapjackWrang_Syurp, parameterName=FJW_Syrup_Inner_Glow_Warp}
  n81 = NamedRerouteUsage()
  n82 = PreSkinnedLocalBounds()
  n83 = Add(
   A=@n82.Min
   B=FunctionInput(Preview=Constant3Vector() {constant=<rgba 0 5 0 0>}) {bUsePreviewValueAsDefault=true, inputName=Object Bounds Min Adjustment, sortPriority=50}
  ) {constB=0}
  n84 = FunctionInput(Preview=Constant3Vector() {constant=<rgba -5 -25 -5 0>}) {bUsePreviewValueAsDefault=true, inputName=Object Bounds Max Adjustment, sortPriority=50}
  n85 = Multiply(A=Subtract(A=Add(A=@n82.Max, B=@n84) {constB=0}, B=@n83)) {constB=0.5}
  n86 = MaterialFunctionCall(Color=NamedRerouteUsage()) {materialFunction=&SplitComponents}
  n87 = LinearInterpolate(A=TransformPosition(None=Add(A=@n83, B=@n85)) {transformType=TRANSFORMPOSSOURCE_World}, B=ObjectPositionWS(), Alpha=@n86.B)
  n88 = Normalize(VectorInput=Subtract(A=@n87, B=CameraPositionWS()))
  n89 = Normalize(VectorInput=CrossProduct(A=@n88, B=Constant3Vector() {constant=<rgba 0 0 1 1>}))
  n90 = MaterialFunctionCall(Float3=@n82.Extents) {materialFunction=&BreakOutFloat3Components}
  n91 = Multiply(
   A=MaterialFunctionCall() {materialFunction=/Engine/Functions/Engine_MaterialFunctions02/WorldPositionOffset/ObjectScale}.Scale X
   B=Max(A=Max(A=@n90.R, B=@n90.R), B=@n90.B)
  )
  n92 = VectorParameter() {group=Wrap_FlapjackWrang_Syurp, parameterName=FJW_Syrup_Inner_Glow__offset, defaultValue=<rgba 0 0 0 1>}
  n93 = LinearInterpolate(
   A=Add(
    A=MaterialFunctionCall!(Plane Normal Axis=@n88, Plane X Axis=@n89, Plane Center=@n87, UV Scale=@n91) {materialFunction=/Game/Athena/Environments/Nexus/Materials/VirtualPlaneCoordinatesAbs}.UVs
    B=ComponentMask(None=@n92.RGB) {g=true, r=true}
   )
   B=@n81
   Alpha=NamedRerouteUsage()
  )
  n94 = ScalarParameter() {defaultValue=1, group=Wrap_FlapjackWrang_Syurp, parameterName=FJW_Syrup_Inner_Glow_Radius_Weapon}
  n95 = NamedRerouteUsage()
  n96 = MaterialFunctionCall(Color=@n95) {materialFunction=&SplitComponents}
  n97 = LinearInterpolate(
   A=ScalarParameter() {defaultValue=1, group=Wrap_FlapjackWrang_Syurp, parameterName=FJW_Syrup_Inner_Glow_Radius_Vehicle}
   B=ScalarParameter() {defaultValue=1, group=Wrap_FlapjackWrang_Syurp, parameterName=FJW_Syrup_Inner_Glow_Radius_VehicleSide}
   Alpha=@n96.G
  )
  n98 = ScalarParameter() {defaultValue=1, group=Wrap_FlapjackWrang_Syurp, parameterName=FJW_Syrup_Inner_Glow_Hardness_Weapon}
  n99 = MaterialFunctionCall(Color=@n95) {materialFunction=&SplitComponents}
  n100 = LinearInterpolate(
   A=ScalarParameter() {defaultValue=1, group=Wrap_FlapjackWrang_Syurp, parameterName=FJW_Syrup_Inner_Glow_Hardness_Vehicle}
   B=ScalarParameter() {defaultValue=1, group=Wrap_FlapjackWrang_Syurp, parameterName=FJW_Syrup_Inner_Glow_Hardness_VehicleSide}
   Alpha=@n99.G
  )
  n101 = SphereMask(
   A=Add(
    A=Multiply(A=@n79, B=@n80)
    B=StaticSwitchParameter(True=@n81, False=@n93) {group=Wrap_FlapjackWrang_Syurp, parameterName=FJW_Syrup_Inner_Glow_Ues_VSPivot, sortPriority=0}
   )
   B=Constant() {r=0.5}
   Radius=LinearInterpolate(
    A=LinearInterpolate(A=@n94, B=@n97, Alpha=@n96.R)
    B=ScalarParameter() {defaultValue=1, group=Wrap_FlapjackWrang_Syurp, parameterName=FJW_Syrup_Inner_Glow_Radius_Mech}
    Alpha=@n96.B
   )
   Hardness=LinearInterpolate(
    A=LinearInterpolate(A=@n98, B=@n100, Alpha=@n99.R)
    B=ScalarParameter() {defaultValue=1, group=Wrap_FlapjackWrang_Syurp, parameterName=FJW_Syrup_Inner_Glow_Hardness_Mech}
    Alpha=@n99.B
   )
  )
  n102 = Saturate(None=Power(Base=OneMinus(None=@n101), Exp=ScalarParameter() {group=Wrap_FlapjackWrang_Syurp, parameterName=FJW_Syrup_Inner_Glow_Falloff}))
  n103 = VectorParameter() {group=Wrap_FlapjackWrang_Syurp, parameterName=FJW_Syrup_Inner__Glow_Color, defaultValue=<rgba 0 0 0 1>}
  n104 = Add(A=@n78.3V, B=Multiply(A=@n70, B=Multiply(A=@n102, B=@n103.RGB)))
  n105 = MaterialFunctionCall(
   Value_Morning=ScalarParameter() {defaultValue=0.5, group=Wrap_FlapjackWrang_TOD, parameterName=FJW_TOD_Mornning}
   Value_Day=ScalarParameter() {defaultValue=1, group=Wrap_FlapjackWrang_TOD, parameterName=FJW_TOD_Day}
   Value_Evening=ScalarParameter() {defaultValue=0.5, group=Wrap_FlapjackWrang_TOD, parameterName=FJW_TOD_Evening}
   Value_Night=ScalarParameter() {defaultValue=0.25, group=Wrap_FlapjackWrang_TOD, parameterName=FJW_TOD_Night}
  ) {materialFunction=&Scalar_Time_of_Day}
  n106 = ScalarParameter() {group=Wrap_FlapjackWrang_TOD, parameterName="FJW_Disable_TOD 0=off 1=active", sortPriority=1}
  n107 = Multiply(
   A=Multiply(
    A=ScalarParameter() {defaultValue=1, group=Wrap_FlapjackWrang, parameterName=FJW_Syrup_Master_Brightness, sortPriority=1}
    B=Multiply(A=@n76, B=@n104)
   )
   B=Saturate(None=Add(A=@n105.Result, B=@n106))
  )
  n108 = SetMaterialAttributes!(
   MaterialAttributes=NamedRerouteUsage()
   BaseColor=LinearInterpolate(A=StaticSwitch(True=@n1, False=@n19.RGB, Value=@n20), B=Constant(), Alpha=NamedRerouteUsage())
   Metallic=StaticSwitch(True=LinearInterpolate(A=@n21, B=@n22, Alpha=@n23), False=@n22, Value=NamedRerouteUsage())
   Specular=LinearInterpolate(A=StaticSwitch(True=@n24, False=@n25, Value=@n26), B=Multiply(A=@n27, B=@n71), Alpha=NamedRerouteUsage())
   Roughness=LinearInterpolate(
    A=StaticSwitch(True=@n72, False=@n73, Value=@n74)
    B=ScalarParameter() {defaultValue=1, group=Wrap_FlapjackWrang, parameterName=FJW_Roughness_Syrup, sortPriority=0}
    Alpha=NamedRerouteUsage()
   )
   EmissiveColor=StaticSwitch(True=Add(A=@n75, B=@n107), False=@n107, Value=NamedRerouteUsage())
   Normal=NamedRerouteUsage()
  ) {attributeSetTypes [
    #1
    #2
    #3
    #4
    #5
    #6
   ]}
  n109 = TextureSample() {samplerType=SAMPLERTYPE_Masks, texture=/Game/Weapons/FORT_Rifles/Textures/T_Scar_CM}
  n110 = FunctionInput(Preview=@n109.R) {bUsePreviewValueAsDefault=true, inputName=Mask Primary, inputType=FunctionInput_Scalar, sortPriority=24, previewValue=<1 0 0 1>}
  n111 = FunctionInput(Preview=@n109.B) {bUsePreviewValueAsDefault=true, inputName=Mask Secondary, inputType=FunctionInput_Scalar, sortPriority=24, previewValue=<1 0 0 1>}
  n112 = Saturate(None=Add(A=Add(A=@n110, B=@n111), B=ScalarParameter() {group=Wrap_FlapjackWrang, parameterName=FJW_Remove_Customization Mask, sortPriority=-10}))
  n113 = NamedRerouteUsage()
  FunctionOutput(None=StaticSwitchParameter(True=BlendMaterialAttributes(A=@n113, B=@n108, Alpha=@n112) {vertexAttributeBlendType=UseA}, False=@n113) {group=Customization_Wrap, parameterName=Wrap_FlapjackWrangler})
  Multiply(A=@n78.Alpha, B=@n70)
  NamedRerouteDeclaration(None=MaterialFunctionCall(Float4=@n2) {materialFunction=&BreakOutFloat4Components}.A) {name=IsCharacter, nodeColor=<rgba 0.8858 1 0.0152 1>}
  n114 = FunctionInput() {bUsePreviewValueAsDefault=true, inputName=Object-SafeZone-Virtual-Pivot, inputType=FunctionInput_Vector4, sortPriority=105}
  n115 = FunctionInput() {bUsePreviewValueAsDefault=true, inputName=Object-SafeZone-Virtual-CH-Switch, inputType=FunctionInput_Scalar, sortPriority=105}
  NamedRerouteDeclaration(None=MaterialFunctionCall!(
    Camera _pos
    Axis
    Scale
    Scale_Multplyer
    Object-SafeZone-Virtual-Pivot=@n114
    Object-SafeZone-Virtual-CH-Switch=@n115
    SinglePlane Normal Axis
    SinglePlane X Axis
    PlaneCenter_distortion
   ) {materialFunction=&/MF_SafeZone_VirtualPlane}.Uvs) {name=SafeZoneVP_UVs, nodeColor=<rgba 1 0 0.6562 1>}
  n116 = MaterialFunctionCall!(Value=@n70) {materialFunction=/Engine/Functions/Engine_MaterialFunctions03/Procedurals/PreparePerturbNormalHQ}
  n117 = Multiply(
   A=Transform(None=MaterialFunctionCall!(
     Bump one pixel right=@n116.Value one pixel right
     Bump center=@n116.Value one pixel right
     Bump one pixel down=@n116.Value one pixel right
     WS Normal
    ) {materialFunction=/Engine/Functions/Engine_MaterialFunctions03/Procedurals/PerturbNormalHQ}.WS Normal) {transformSourceType=TRANSFORMSOURCE_World, transformType=TRANSFORM_Tangent}
   B=Constant3Vector() {constant=<rgba 1 1 0 0>}
  )
  n118 = StaticSwitch(
   True=LinearInterpolate(A=NamedRerouteUsage(), B=NamedRerouteUsage(), Alpha=NamedRerouteUsage())
   False=NamedRerouteUsage()
   Value=NamedRerouteUsage()
  )
  NamedRerouteDeclaration(None=LinearInterpolate(
    A=@n118
    B=Add(A=@n118, B=@n117)
    Alpha=ScalarParameter() {defaultValue=1, group=Wrap_FlapjackWrang, parameterName=FJW_Syrup_Normal_Intensty}
   )) {name=normals, nodeColor=<rgba 0 1 0.6719 1>}
  NamedRerouteDeclaration(None=@n78.Alpha) {name=Syrup_Mask, nodeColor=<rgba 1 0 0.4922 1>}
  n119 = GetMaterialAttributes(None=FunctionInput() {inputName=Material_Root, inputType=FunctionInput_MaterialAttributes, sortPriority=-20}) {attributeGetTypes [
    #6
    #1
    #4
    #3
    #2
    #5
   ]}
  NamedRerouteDeclaration(None=@n119.Normal) {name=Roots_Normals, nodeColor=<rgba 0.7344 1 0 1>}
  n120 = Multiply(A=NamedRerouteUsage(), B=NamedRerouteUsage())
  n121 = NamedRerouteUsage()
  n122 = ChannelMaskParameter(None=TextureSample(UVs=Add(A=@n120, B=@n121), Tex=@n43).RGBA) {group=Wrap_FlapjackWrang, parameterName=FJW_SyrupWarp__Tex_maskChannel}
  NamedRerouteDeclaration(None=Multiply(A=Add(A=@n122.R) {constB=-0.5}, B=ScalarParameter() {group=Wrap_FlapjackWrang, parameterName=FJW_SyrupDripWarp_Amount}) {constB=2}) {name=Warp_Noise_Drip, nodeColor=<rgba 0 1 0.3906 1>}
  NamedRerouteDeclaration(None=@n2) {name=IsState, nodeColor=<rgba 0.1497 1 0.1633 1>}
  NamedRerouteDeclaration(None=FunctionInput(Preview=Constant2Vector()) {bUsePreviewValueAsDefault=true, inputName=IsState_B, inputType=FunctionInput_Vector2, sortPriority=4}) {name=IsState_B, nodeColor=<rgba 0.1497 1 0.1633 1>}
  n123 = VectorParameter() {group=Wrap_FlapjackWrang_Uvs, parameterName=FJW_SyrupDrip_TileRate_Weapon, defaultValue=<rgba 1 1 1 1>}
  n124 = Multiply(A=@n123.RGB, B=@n123.RGB)
  n125 = VectorParameter() {group=Wrap_FlapjackWrang_Uvs, parameterName=FJW_SyrupDrip_TileRate_Vehicle, defaultValue=<rgba 1 1 1 1>}
  n126 = VectorParameter() {group=Wrap_FlapjackWrang_Uvs, parameterName=FJW_SyrupDrip_TileRate_VehicleSide, defaultValue=<rgba 1 1 1 1>}
  n127 = MaterialFunctionCall(Float4=NamedRerouteUsage()) {materialFunction=&BreakOutFloat4Components}
  n128 = LinearInterpolate(A=Multiply(A=@n125.RGB, B=@n125.RGB), B=Multiply(A=@n126.RGB, B=@n126.RGB), Alpha=@n127.G)
  n129 = VectorParameter() {group=Wrap_FlapjackWrang_Uvs, parameterName=FJW_SyrupDrip_TileRate_Mech, defaultValue=<rgba 1 1 1 1>}
  n130 = Multiply(A=@n129.RGB, B=@n129.RGB)
  n131 = VectorParameter() {group=Wrap_FlapjackWrang_Uvs, parameterName=FJW_SyrupDrip_TileRate_Character, defaultValue=<rgba 1 1 1 1>}
  n132 = Multiply(A=@n131.RGB, B=@n131.RGB)
  n133 = VectorParameter() {group=Wrap_FlapjackWrang_Uvs, parameterName=FJW_SyrupDrip_TileRate_Backpack, defaultValue=<rgba 1 1 1 1>}
  n134 = MaterialFunctionCall(Float2=NamedRerouteUsage()) {materialFunction=&BreakOutFloat2Components}
  n135 = LinearInterpolate(
   A=LinearInterpolate(A=LinearInterpolate(A=@n124, B=@n128, Alpha=@n127.R), B=LinearInterpolate(A=@n130, B=@n132, Alpha=@n127.A), Alpha=@n127.B)
   B=Multiply(A=@n133.RGB, B=@n133.RGB)
   Alpha=@n134.R
  )
  n136 = VectorParameter() {group=Wrap_FlapjackWrang_Uvs, parameterName=FJW_SyrupDrip_TileRate_Pickaxe, defaultValue=<rgba 1 1 1 1>}
  n137 = Multiply(A=@n136.RGB, B=@n136.RGB)
  NamedRerouteDeclaration(None=ComponentMask(None=LinearInterpolate(A=@n135, B=@n137, Alpha=@n134.G)) {g=true, r=true}) {name=UVs_Syrup_TileRate, nodeColor=<rgba 0 0.7188 1 1>}
  n138 = FunctionInput(Preview=@n109.G) {bUsePreviewValueAsDefault=true, inputName=Mask Crevise, inputType=FunctionInput_Scalar, sortPriority=24, previewValue=<1 0 0 1>}
  NamedRerouteDeclaration(None=@n138) {name=Curvature, nodeColor=<rgba 0.6172 1 0 1>}
  n139 = VectorParameter() {group=Wrap_FlapjackWrang_Uvs, parameterName=FJW_SyrupDripWarp_Speed_Vehicle}
  n140 = VectorParameter() {group=Wrap_FlapjackWrang_Uvs, parameterName=FJW_SyrupDripWarp_Speed_VehicleSide, defaultValue=<rgba 0 1 0 1>}
  n141 = MaterialFunctionCall(Color=NamedRerouteUsage()) {materialFunction=&SplitComponents}
  n142 = LinearInterpolate(
   A=LinearInterpolate(
    A=VectorParameter() {group=Wrap_FlapjackWrang_Uvs, parameterName=FJW_SyrupDripWarp_Speed_Weapon}.RGB
    B=LinearInterpolate(A=@n139.RGB, B=@n140.RGB, Alpha=@n141.G)
    Alpha=@n141.R
   )
   B=VectorParameter() {group=Wrap_FlapjackWrang_Uvs, parameterName=FJW_SyrupDripWarp_Speed_Mech}.RGB
   Alpha=@n141.B
  )
  n143 = Time()
  NamedRerouteDeclaration(None=ComponentMask(None=Multiply(A=@n142, B=@n143)) {g=true, r=true}) {name=Drip_Wrap_Speed, nodeColor=<rgba 1 0.3281 0 1>}
  n144 = ScalarParameter() {defaultValue=1, group=Wrap_FlapjackWrang, parameterName=FJW_SyrupWarp_Scale}
  n145 = VectorParameter() {group=Wrap_FlapjackWrang_Uvs, parameterName=FJW_SyrupWarp_Speed_Vehicle}
  n146 = VectorParameter() {group=Wrap_FlapjackWrang_Uvs, parameterName=FJW_SyrupWarp_Speed_VehicleSide, defaultValue=<rgba 0 1 0 1>}
  n147 = LinearInterpolate(
   A=LinearInterpolate(
    A=VectorParameter() {group=Wrap_FlapjackWrang_Uvs, parameterName=FJW_SyrupWarp_Speed_Weapon}.RGB
    B=LinearInterpolate(A=@n145.RGB, B=@n146.RGB, Alpha=@n86.G)
    Alpha=@n86.R
   )
   B=VectorParameter() {group=Wrap_FlapjackWrang_Uvs, parameterName=FJW_SyrupWarp_Speed_Mech}.RGB
   Alpha=@n86.B
  )
  n148 = Time()
  n149 = MaterialFunctionCall(Color=NamedRerouteUsage()) {materialFunction=&SplitComponents}
  n150 = LinearInterpolate(
   A=ScalarParameter() {group=Wrap_FlapjackWrang_Uvs, parameterName=FJW_SyrupWarp_Rotation_Weapon}
   B=LinearInterpolate(
    A=ScalarParameter() {group=Wrap_FlapjackWrang_Uvs, parameterName=FJW_SyrupWarp_Rotation_Vehicle}
    B=ScalarParameter() {group=Wrap_FlapjackWrang_Uvs, parameterName=FJW_SyrupWarp_Rotation_VehicleSide}
    Alpha=@n149.G
   )
   Alpha=@n149.R
  )
  n151 = ScalarParameter() {group=Wrap_FlapjackWrang_Uvs, parameterName=FJW_SyrupWarp_Rotation_Mech}
  n152 = Add(
   A=ComponentMask(None=Multiply(A=@n147, B=@n148)) {g=true, r=true}
   B=MaterialFunctionCall(UVs=NamedRerouteUsage(), "Rotation Angle (0-1)"=LinearInterpolate(A=@n150, B=@n151, Alpha=@n149.B)) {materialFunction=&CustomRotator}.Rotated Values
  )
  n153 = ChannelMaskParameter(None=TextureSample(UVs=Multiply(A=@n144, B=@n152), Tex=@n43).RGBA) {group=Wrap_FlapjackWrang, parameterName=FJW_SyrupWarp__Tex_maskChannel}
  NamedRerouteDeclaration(None=Multiply(A=Add(A=@n153.R) {constB=-0.5}, B=ScalarParameter() {group=Wrap_FlapjackWrang, parameterName=FJW_SyrupWarp_Amount}) {constB=2}) {name=Warp_Noise, nodeColor=<rgba 0 1 0.3906 1>}
  n154 = TextureSample() {samplerType=SAMPLERTYPE_Normal, texture=/Game/Weapons/FORT_Rifles/T_SCAR_N}
  n155 = GetMaterialAttributes(None=FunctionInput(Preview=SetMaterialAttributes!(MaterialAttributes, Normal=@n154.RGB) {attributeSetTypes [#6]}) {inputName=Material, inputType=FunctionInput_MaterialAttributes, sortPriority=-20}) {attributeGetTypes [
    #6
    #1
    #4
    #3
    #2
    #5
   ]}
  NamedRerouteDeclaration(None=@n155.MaterialAttributes) {name=Root, nodeColor=<rgba 1 0.7266 0 1>}
  NamedRerouteDeclaration(None=@n14.UVs No Offsets/Tiling) {name=Uvs_NoOffset, nodeColor=<rgba 0 0.2734 1 1>}
  n156 = ScalarParameter() {defaultValue=1, group=Wrap_FlapjackWrang_Uvs, parameterName=FJW_SyrupDripWarp_TileRate_Weapon}
  n157 = MaterialFunctionCall(Color=NamedRerouteUsage()) {materialFunction=&SplitComponents}
  n158 = LinearInterpolate(
   A=ScalarParameter() {defaultValue=1, group=Wrap_FlapjackWrang_Uvs, parameterName=FJW_SyrupDripWarp_TileRate_Vehicle}
   B=ScalarParameter() {defaultValue=1, group=Wrap_FlapjackWrang_Uvs, parameterName=FJW_SyrupDripWarp_TileRate_VehicleSide}
   Alpha=@n157.G
  )
  NamedRerouteDeclaration(None=LinearInterpolate(
    A=LinearInterpolate(A=@n156, B=@n158, Alpha=@n157.R)
    B=ScalarParameter() {defaultValue=1, group=Wrap_FlapjackWrang_Uvs, parameterName=FJW_SyrupDripWarp_TileRate_Mech}
    Alpha=@n157.B
   )) {name=Drip_Wrap_TileRate, nodeColor=<rgba 1 0.3281 0 1>}
  NamedRerouteDeclaration(None=@n155.Normal) {name=Root_Normal, nodeColor=<rgba 1 0 0.2578 1>}
  NamedRerouteDeclaration(None=StaticBoolParameter() {group=Customization, parameterName=TopSideProjection_V02}) {name=Top_Side_State, nodeColor=<rgba 0 1 0.6484 1>}
  NamedRerouteDeclaration(None=@n155.BaseColor) {name=Root_BaseColor, nodeColor=<rgba 0.8984 1 0 1>}
  NamedRerouteDeclaration(None=@n155.Roughness) {name=Root_Roughness, nodeColor=<rgba 0.7813 0 1 1>}
  NamedRerouteDeclaration(None=@n155.Specular) {name=Root_Spec, nodeColor=<rgba 0 1 0.4609 1>}
  NamedRerouteDeclaration(None=@n155.Metallic) {name=Root_Metalic, nodeColor=<rgba 1 0.1406 0 1>}
  NamedRerouteDeclaration(None=@n155.EmissiveColor) {name=Root_Emissive, nodeColor=<rgba 0 0.1797 1 1>}
  Saturate(None=@n70)
  ScalarParameter() {defaultValue=1, group=Wrap_FlapjackWrang_Syurp, parameterName=FJW_Syrup_Inner_Glow_Radius}
  ScalarParameter() {group=Wrap_FlapjackWrang_Syurp, parameterName=FJW_Syrup_Inner_Glow_Hardness}
  StaticSwitchParameter() {group=Customization, parameterName=TopSideProjection_V02}
 ]}
