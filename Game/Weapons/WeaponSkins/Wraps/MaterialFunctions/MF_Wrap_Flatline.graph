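# Material function for the optional Flatline cosmetic wrap, blended over the input material by a customization mask and static switch.
# It layers a world-aligned grid, a side-projected graph revealed by a panning gradient, and Fresnel, with projection settings chosen per item type.
# A time-since-event parameter temporarily switches color, graph channel and emissive strength, then fades back.
# A top/side-projected mask finally overlays a separate non-emissive surface.
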MaterialFunction /Game/Weapons/WeaponSkins/Wraps/MaterialFunctions/MF_Wrap_Flatline
graph {nodes [
  n1 = GetMaterialAttributes(None=FunctionInput() {inputName=Material, inputType=FunctionInput_MaterialAttributes, sortPriority=-20}) {attributeGetTypes [#1]}
  n2 = VectorParameter() {group=Wrap_Flatline, parameterName=FL_BaseColor-----, defaultValue=<rgba 0.01 0.01 0.01 1>}
  n3 = Subtract(
   A=Time()
   B=ScalarParameter() {defaultValue=-9999999, group=Wrap_Flatline_Reactive, parameterName=TimeFrom-BP----------------, sortPriority=600}
  )
  n4 = Max(A=@n3) {constB=0}
  n5 = ScalarParameter() {defaultValue=0.8, group=Wrap_Flatline_Reactive, parameterName=FL_Time_Duration, sortPriority=601}
  n6 = Multiply(A=@n5)
  n7 = Saturate(None=Divide(A=Subtract(A=@n4, B=@n6) {desc="how long the reaction will stay at 0 (100%)"}, B=Multiply(A=@n5) {constB=0.25}) {desc=the length of time the reaction takes to fade})
  n8 = LinearInterpolate(
   A=VectorParameter() {group=Wrap_Flatline_Reactive, parameterName=FL_BaseColor_Reactive, defaultValue=<rgba 1 0 0 1>}.RGB
   B=VectorParameter() {group=Wrap_Flatline, parameterName=FL_Graph_BaseColor-----, defaultValue=<rgba 0 1 0 1>}.RGB
   Alpha=@n7
  )
  n9 = Multiply(A=ScalarParameter() {defaultValue=0.1, group=Wrap_Flatline, parameterName=FL_Grid_Color_Intensity, sortPriority=1}, B=@n8)
  n10 = ScalarParameter() {defaultValue=1, group=Wrap_Flatline, parameterName=FL_Grid_TileRate_Vehicle}
  n11 = ScalarParameter() {defaultValue=1, group=Wrap_Flatline, parameterName=FL_Grid_TileRate_VehicleSide}
  n12 = Constant()
  n13 = MaterialFunctionCall(
   R=FunctionInput(Preview=Constant()) {inputName=isVehicle, inputType=FunctionInput_Scalar, sortPriority=5}
   G=FunctionInput(Preview=@n12) {bUsePreviewValueAsDefault=true, inputName=Side Project Vehicle, inputType=FunctionInput_Scalar}
   B=FunctionInput(Preview=@n12) {bUsePreviewValueAsDefault=true, inputName=isMech, inputType=FunctionInput_Scalar, sortPriority=4}
  ) {materialFunction=&AppendMany}
  n14 = MaterialFunctionCall(Color=@n13.RGB) {materialFunction=&SplitComponents}
  n15 = LinearInterpolate(
   A=LinearInterpolate(
    A=ScalarParameter() {defaultValue=1, group=Wrap_Flatline, parameterName=FL_Grid_TileRate_Weapon}
    B=LinearInterpolate(A=@n10, B=@n11, Alpha=@n14.G)
    Alpha=@n14.R
   )
   B=ScalarParameter() {defaultValue=1, group=Wrap_Flatline, parameterName=FL_Grid_TileRate_Mech}
   Alpha=@n14.B
  )
  n16 = VertexInterpolator(VS=PreSkinnedPosition())
  n17 = PreSkinnedLocalBounds()
  n18 = FunctionInput(Preview=Divide(A=Subtract(A=@n16.PS, B=@n17.Min), B=Subtract(A=@n17.Max, B=@n17.Max))) {bUsePreviewValueAsDefault=true, inputName=Calcated_Bounds, sortPriority=11}
  n19 = FunctionInput(Preview=VertexInterpolator(VS=PreSkinnedNormal()).PS) {bUsePreviewValueAsDefault=true, inputName=Pre-Skinned Local Normal, sortPriority=10}
  n20 = MaterialFunctionCall(Color=Constant3Vector() {constant=<rgba 0 -0.25 -0.25 1>}) {materialFunction=&SplitComponents}
  n21 = ChannelMaskParameter(None=MaterialFunctionCall(
    Export Float 4=StaticBool() {value=true}
    TextureObject=TextureObjectParameter() {group=Wrap__Flatline_Textures, parameterName=FL_GridTexture, texture=&T_BlacksRGB}
    TextureSize=Constant() {r=1}
    WorldPosition=Multiply(A=@n15, B=@n18)
    World Space Normal=@n19
    RotZAxis=@n20.B
    RotXAxis=@n20.B
    RotYAxis=@n20.B
    ProjectionTransitionContrast=ScalarParameter() {defaultValue=1, group=Wrap_Flatline, parameterName=FL_Grid_Blend_Contrast-----}
   ) {materialFunction=&/WorldAlignedTexture}.XYZ Texture) {group=Wrap_Flatline, maskChannel=Green, parameterName=FL_Grid_Channel, defaultValue=<rgba 0 1 0 0>}
  n22 = Multiply(A=@n8, B=ScalarParameter() {defaultValue=2, group=Wrap_Flatline, parameterName=FL_Graph_Color_Intensity, sortPriority=1})
  n23 = MaterialFunctionCall(Color=@n13.RGB) {materialFunction=&SplitComponents}
  n24 = ScalarParameter() {defaultValue=0.25, group=Wrap_Flatline, parameterName=FL_Graph_Rotation_Mech}
  n25 = ScalarParameter() {group=Wrap_Flatline, parameterName=FL_Graph_Rotation_Character}
  n26 = NamedRerouteUsage()
  n27 = VectorParameter() {group=Wrap_Flatline, parameterName=FL_Graph_Offset_Vehicle, defaultValue=<rgba 0 0 0 1>}
  n28 = VectorParameter() {group=Wrap_Flatline, parameterName=FL_Graph_Offset_VehicleSide, defaultValue=<rgba 0 0 0 1>}
  n29 = MaterialFunctionCall(Color=@n13.RGB) {materialFunction=&SplitComponents}
  n30 = LinearInterpolate(
   A=LinearInterpolate(
    A=VectorParameter() {group=Wrap_Flatline, parameterName=FL_Graph_Offset_Weapon, defaultValue=<rgba 0 0 0 1>}.RGB
    B=LinearInterpolate(A=@n27.RGB, B=@n28.RGB, Alpha=@n29.G)
    Alpha=@n29.R
   )
   B=VectorParameter() {group=Wrap_Flatline, parameterName=FL_Graph_Offset_Mech, defaultValue=<rgba 0 0 0 1>}.RGB
   Alpha=@n29.B
  )
  n31 = VectorParameter() {group=Wrap_Flatline, parameterName=FL_Graph_Offset_Character, defaultValue=<rgba 0 0 0 1>}
  n32 = NamedRerouteUsage()
  n33 = MaterialFunctionCall!(
   For_Safe_Zones
   Normal_Map
   Caculated_Bounds=@n18
   Pre_Skinned_Normals=@n19
   Is Vehicle=@n23.R
   Is VehicleSide=@n23.R
   Is Mech=@n23.R
   Vehicle_Rotation=ScalarParameter() {group=Wrap_Flatline, parameterName=FL_Graph_Rotation_Vehicle}
   UV_Rotation Bypass
   Mech_Rotation=LinearInterpolate(
    A=LinearInterpolate(A=@n24, B=@n25, Alpha=@n26)
    B=ScalarParameter() {group=Wrap_Flatline, parameterName=FL_Graph_Rotation_Backpack}
    Alpha=NamedRerouteUsage()
   )
   Uvs_Offset=LinearInterpolate(
    A=LinearInterpolate(A=@n30, B=@n31.RGB, Alpha=@n32)
    B=VectorParameter() {group=Wrap_Flatline, parameterName=FL_Graph_Offset_Backpack, defaultValue=<rgba 0 0 0 1>}.RGB
    Alpha=NamedRerouteUsage()
   )
   Offset-after rotation
   Uvs_TileRate
  ) {materialFunction=&/MF_Projection_Side_Normal}
  n34 = TextureSample(
   UVs=@n33.UVs
   Tex=TextureObjectParameter() {group=Wrap__Flatline_Textures, parameterName=FL_BC_GraphPattern, texture=&T_BlacksRGB}
  )
  n35 = ChannelMaskParameter(None=@n34.RGBA) {group=Wrap_Flatline, maskChannel=Alpha, parameterName=FL_Graph_Channel, defaultValue=<rgba 0 0 0 1>}
  n36 = Multiply(A=@n33.UVs, B=ScalarParameter() {defaultValue=1, group=Wrap_Flatline, parameterName=FL_Gradient_TileRate})
  n37 = VectorParameter() {group=Wrap_Flatline, parameterName=FL_Gradient_Animation_Vehicle, defaultValue=<rgba 0 0 0 1>}
  n38 = VectorParameter() {group=Wrap_Flatline, parameterName=FL_Gradient_Animation_VehicleSide, defaultValue=<rgba 0 0 0 1>}
  n39 = MaterialFunctionCall(Color=@n13.RGB) {materialFunction=&SplitComponents}
  n40 = LinearInterpolate(
   A=LinearInterpolate(
    A=VectorParameter() {group=Wrap_Flatline, parameterName=FL_Gradient_Animation_Weapon, defaultValue=<rgba 0 0 0 1>}.RGB
    B=LinearInterpolate(A=@n37.RGB, B=@n38.RGB, Alpha=@n39.G)
    Alpha=@n39.R
   )
   B=VectorParameter() {group=Wrap_Flatline, parameterName=FL_Gradient_Animation_Mech-----, defaultValue=<rgba 0 0 0 1>}.RGB
   Alpha=@n39.B
  )
  n41 = VectorParameter() {group=Wrap_Flatline, parameterName=FL_Gradient_Animation_Character-----, defaultValue=<rgba 0 0 0 1>}
  n42 = NamedRerouteUsage()
  n43 = ComponentMask(None=LinearInterpolate(
    A=LinearInterpolate(A=@n40, B=@n41.RGB, Alpha=@n42)
    B=VectorParameter() {group=Wrap_Flatline, parameterName=FL_Gradient_Animation_Backpack-----, defaultValue=<rgba 0 0 0 1>}.RGB
    Alpha=NamedRerouteUsage()
   )) {g=true, r=true}
  n44 = AppendVector(A=ScalarParameter() {group=Wrap_Flatline_Reactive, parameterName=FL_Gradient_PanSpeed_OverTime}, B=Constant())
  n45 = Divide(A=@n3, B=@n5)
  n46 = Add(
   A=Multiply(A=Time(), B=Add(A=@n43, B=@n44))
   B=LinearInterpolate(A=ScalarParameter() {defaultValue=1, group=Wrap_Flatline_Reactive, parameterName=FL_Gradient_PanSpeed_Reacitve}, Alpha=Saturate(None=@n45)) {constA=0.5, constB=0}
  ) {constB=0.5}
  n47 = ChannelMaskParameter(None=TextureSample(
    UVs=Add(A=@n36, B=@n46)
    Tex=TextureObjectParameter() {group=Wrap__Flatline_Textures, parameterName=FL_Gradient, texture=&T_BlacksRGB}
   ).RGBA) {group=Wrap_Flatline, maskChannel=Green, parameterName=FL_Gradient_Channel, defaultValue=<rgba 0 0 1 0>}
  n48 = Multiply(
   A=LinearInterpolate(A=ChannelMaskParameter(None=@n34.RGBA) {group=Wrap_Flatline_Reactive, parameterName=FL_Graph_ChannelReactive}.R, B=@n35.R, Alpha=@n7)
   B=@n47.R
  )
  n49 = Multiply(A=@n48, B=Add(A=@n48, B=@n47.R))
  n50 = Constant()
  n51 = MaterialFunctionCall(Color=Abs(None=@n19)) {materialFunction=&SplitComponents}
  n52 = MaterialFunctionCall(Color=@n13.RGB) {materialFunction=&SplitComponents}
  n53 = LinearInterpolate(A=Add(A=@n51.R, B=@n51.R), B=Add(A=@n51.B, B=@n51.B), Alpha=@n52.R)
  n54 = Add(A=@n51.R, B=@n51.R)
  n55 = ScalarParameter() {defaultValue=0.1, group=Wrap_Flatline, parameterName=FL_Mask_Contrast, sortPriority=51}
  n56 = Saturate(None=MaterialFunctionCall(In=LinearInterpolate(A=@n53, B=@n54, Alpha=@n52.B), Contrast=@n55) {materialFunction=&CheapContrast}.Result)
  n57 = ScalarParameter() {group=Wrap_Flatline, parameterName=FL_Mask_AltAxisActive -----, sortPriority=50}
  n58 = Constant()
  n59 = MaterialFunctionCall(Color=Abs(None=@n19)) {materialFunction=&SplitComponents}
  n60 = LinearInterpolate(A=@n59.R, B=LinearInterpolate(A=@n59.B, B=@n59.B, Alpha=@n52.G), Alpha=@n52.R)
  n61 = Saturate(None=MaterialFunctionCall(In=LinearInterpolate(A=@n60, B=@n59.G, Alpha=@n52.B), Contrast=@n55) {materialFunction=&CheapContrast}.Result)
  n62 = ScalarParameter() {group=Wrap_Flatline, parameterName=FL_Mask_Z+AltAxisActive -----, sortPriority=50}
  n63 = Multiply(
   A=Multiply(A=ScalarParameter() {defaultValue=1, group=Wrap_Flatline, parameterName=FL_Graph_Intensity}, B=Saturate(None=@n49))
   B=LinearInterpolate(A=LinearInterpolate(A=@n50, B=@n56, Alpha=@n57), B=LinearInterpolate(A=@n58, B=@n61, Alpha=@n62), Alpha=@n62)
  )
  n64 = ScalarParameter() {defaultValue=2, group=Wrap_Flatline, parameterName=FL__Fresnel, sortPriority=1}
  n65 = ScalarParameter() {defaultValue=8, group=Wrap_Flatline_Reactive, parameterName=FL__Fresnel_Reactive, sortPriority=1}
  n66 = MaterialFunctionCall(
   In=Fresnel(ExponentIn=LinearInterpolate(A=@n64, B=@n65), Normal=Transform(None=@n1.Normal))
   Contrast=ScalarParameter() {group=Wrap_Flatline, parameterName=FL__FresnelContrast, sortPriority=1}
  ) {materialFunction=&CheapContrast}
  n67 = LinearInterpolate(
   A=LinearInterpolate(A=LinearInterpolate(A=@n2.RGB, B=@n9, Alpha=@n21.R), B=Multiply(A=@n22, B=@n63), Alpha=@n63)
   B=Multiply(
    A=ScalarParameter() {defaultValue=0.5, group=Wrap_Flatline, parameterName=FL__FresnelColorAmount, sortPriority=1}
    B=Multiply(A=@n66.Result, B=@n8)
   )
   Alpha=@n66.Result
  )
  n68 = ScalarParameter() {defaultValue=0.5, group=Wrap_Flatline, parameterName=FL__Specular_Grid, sortPriority=1}
  n69 = ScalarParameter() {defaultValue=0.5, group=Wrap_Flatline, parameterName=FL__Specular_GridHigh, sortPriority=1}
  n70 = ScalarParameter() {defaultValue=0.15, group=Wrap_Flatline, parameterName=FL__Roughness_Grid, sortPriority=1}
  n71 = ScalarParameter() {defaultValue=2, group=Wrap_Flatline_Reactive, parameterName=FL_Emissive_Amount_Reactive}
  n72 = ScalarParameter() {defaultValue=1, group=Wrap_Flatline, parameterName=FL_Emissive_Amount}
  n73 = SetMaterialAttributes!(
   MaterialAttributes=@n1.MaterialAttributes
   BaseColor=@n67
   Metallic=ScalarParameter() {group=Wrap_Flatline, parameterName=FL__Metallic_Base, sortPriority=1}
   Specular=Add(A=@n66.Result, B=LinearInterpolate(A=@n68, B=@n69, Alpha=@n21.R))
   Roughness=Add(A=Multiply(A=@n70, B=@n21.R), B=ScalarParameter() {defaultValue=0.3, group=Wrap_Flatline, parameterName=FL__Roughness_Base, sortPriority=1})
   EmissiveColor=Multiply(A=@n67, B=LinearInterpolate(A=@n71, B=@n72, Alpha=@n7))
  ) {attributeSetTypes [
    #2
    #3
    #4
    #5
    #6
   ]}
  n74 = GetMaterialAttributes(None=@n73) {attributeGetTypes [
    #2
    #3
    #4
    #5
    #6
   ]}
  n75 = VectorParameter() {group=Wrap_Flatline, parameterName=FL_TS_BaseColor-----, sortPriority=1, defaultValue=<rgba 0.01 0.01 0.01 1>}
  n76 = MaterialFunctionCall(Color=@n13.RGB) {materialFunction=&SplitComponents}
  n77 = VectorParameter() {group=Wrap_Flatline_TopSideProjection, parameterName="FL_TS_Fill_alternate_Axis_Color_A=Roughness 01", sortPriority=33}
  n78 = VectorParameter() {
   group=Wrap_Flatline_TopSideProjection
   parameterName=FL_TS_Offset_Weapon_01---------------------
   sortPriority=10
   defaultValue=<rgba 0 0 0 1>
  }
  n79 = VectorParameter() {group=Wrap_Flatline_TopSideProjection, parameterName=FL_TS_Offset_Vehicle_01, sortPriority=11, defaultValue=<rgba 0 0 0 1>}
  n80 = VectorParameter() {group=Wrap_Flatline_TopSideProjection, parameterName=FL_TS_Offset_Vehicle_Side_01, sortPriority=11, defaultValue=<rgba 0 0 0 1>}
  n81 = VectorParameter() {group=Wrap_Flatline_TopSideProjection, parameterName=FL_TS_Offset_Mech_01, sortPriority=12, defaultValue=<rgba 0 0 0 1>}
  n82 = LinearInterpolate(A=LinearInterpolate(A=@n78.RGB, B=LinearInterpolate(A=@n79.RGB, B=@n80.RGB, Alpha=@n76.G), Alpha=@n76.R), B=@n81.RGB, Alpha=@n76.B)
  n83 = VectorParameter() {group=Wrap_Flatline_TopSideProjection, parameterName="FL_TS_Scale_Weapon_01 A=uniform", sortPriority=30, defaultValue=<rgba 1 1 1 1>}
  n84 = VectorParameter() {group=Wrap_Flatline_TopSideProjection, parameterName="FL_TS_Scale_Vehicle_01 A=uniform", sortPriority=31, defaultValue=<rgba 1 1 1 1>}
  n85 = VectorParameter() {group=Wrap_Flatline_TopSideProjection, parameterName="FL_TS_Scale_Vehicle_Side_01 A=uniform", sortPriority=31, defaultValue=<rgba 1 1 1 1>}
  n86 = LinearInterpolate(
   A=Multiply(A=@n83.RGB, B=@n83.RGB)
   B=LinearInterpolate(A=Multiply(A=@n84.RGB, B=@n84.RGB), B=Multiply(A=@n85.RGB, B=@n85.RGB), Alpha=@n76.G)
   Alpha=@n76.R
  )
  n87 = VectorParameter() {group=Wrap_Flatline_TopSideProjection, parameterName="FL_TS_Scale_Mech_01 A=uniform", defaultValue=<rgba 1 1 1 1>}
  n88 = Multiply(A=@n87.RGB, B=@n87.RGB)
  n89 = VectorParameter() {
   group=Wrap_Flatline_TopSideProjection
   parameterName=FL_TS-Rotation-Weapon-01---------------------
   sortPriority=1
   defaultValue=<rgba 0 0 -0.25 1>
  }
  n90 = VectorParameter() {group=Wrap_Flatline_TopSideProjection, parameterName=FL_TS_Rotation_Vehicle_01, sortPriority=2, defaultValue=<rgba 0 0 0 1>}
  n91 = VectorParameter() {group=Wrap_Flatline_TopSideProjection, parameterName=FL_Rotation_Vehicle_Side_01, sortPriority=2, defaultValue=<rgba 0 0 0 1>}
  n92 = LinearInterpolate(A=@n90.RGB, B=@n91.RGB, Alpha=@n76.G)
  n93 = VectorParameter() {group=Wrap_Flatline_TopSideProjection, parameterName=FL_TS_Rotation_Mech_01, sortPriority=3, defaultValue=<rgba 0.25 0.25 0 1>}
  n94 = MaterialFunctionCall!(
   IsCharacter
   IsStatsB
   Texture=TextureObjectParameter() {group=Wrap__Flatline_Textures, parameterName=FL_TS_Mask, texture=&T_BlacksRGB}
   Rotation_Axis=LinearInterpolate(A=LinearInterpolate(A=@n89.RGB, B=@n92, Alpha=@n76.R), B=@n93.RGB, Alpha=@n76.B)
   Caculated_Bounds=Multiply(A=Add(A=@n18, B=@n82), B=LinearInterpolate(A=@n86, B=@n88, Alpha=@n76.B))
   Pre_Skinned_Normals=@n19
   Blend_Contrast=LinearInterpolate(
    A=ScalarParameter() {defaultValue=1, group=Wrap_Flatline_TopSideProjection, parameterName=FL_TS_Blend_Contrast, sortPriority=33}
    B=ScalarParameter() {defaultValue=2, group=Wrap_Flatline_TopSideProjection, parameterName=FL_TS_Blend_Contrast_Mech, sortPriority=33}
    Alpha=@n76.B
   )
   Fill_alternate_Axis_Opacity=ScalarParameter() {group=Wrap_Flatline_TopSideProjection, parameterName=FL_TS_Fill_alternate_Axis_Opacity 01, sortPriority=33}
   "Fill_alternate_Axis_Color_A=Roughness"=AppendVector(A=@n77.RGB, B=@n77.RGB)
   Is_Vehicle=@n76.R
   Is_Mech=@n76.B
   Enable_Fill_Z_Weapon
   Fill_Z_Weapon_opacity
   Fill_Z_Weapon_Color
  ) {materialFunction=&/MF_TopsSideProjection_Base}
  n95 = ChannelMaskParameter(None=@n94.Combined) {group=Wrap_Flatline, parameterName=FL_TS_MaskChannel, sortPriority=1}
  n96 = ScalarParameter() {group=Wrap_Flatline, parameterName=FL_TS_Roughness, sortPriority=1}
  n97 = ScalarParameter() {defaultValue=1, group=Wrap_Flatline, parameterName=FL_TS_Roughness_BaseHigh, sortPriority=1}
  n98 = ChannelMaskParameter(None=@n94.Combined) {group=Wrap_Flatline, maskChannel=Alpha, parameterName=FL_TS_BaseRoughness, defaultValue=<rgba 0 0 0 1>}
  n99 = SetMaterialAttributes(
   MaterialAttributes=@n74.MaterialAttributes
   BaseColor=LinearInterpolate(A=@n74.BaseColor, B=@n75.RGB, Alpha=@n95.R)
   Metallic=LinearInterpolate(A=@n74.Metallic, B=ScalarParameter() {group=Wrap_Flatline, parameterName=FL_TS_Metallic, sortPriority=1}, Alpha=@n95.R)
   Specular=LinearInterpolate(A=@n74.Specular, B=ScalarParameter() {group=Wrap_Flatline, parameterName=FL_TS_Specular, sortPriority=1}, Alpha=@n95.R)
   Roughness=LinearInterpolate(A=@n74.Roughness, B=LinearInterpolate(A=@n96, B=@n97, Alpha=@n98.R), Alpha=@n95.R)
   EmissiveColor=LinearInterpolate(A=@n74.EmissiveColor, B=Constant(), Alpha=@n95.R)
  ) {attributeSetTypes [
    #2
    #3
    #4
    #5
    #6
   ]}
  n100 = Saturate(None=Add(
    A=FunctionInput() {bUsePreviewValueAsDefault=true, inputName="Customization Mask (_CM B)", inputType=FunctionInput_Scalar, sortPriority=24}
    B=ScalarParameter() {group=Wrap_Flatline, parameterName=FL_Remove_Customization Mask, sortPriority=-10}
   ))
  FunctionOutput(None=StaticSwitchParameter(True=BlendMaterialAttributes(A=@n1.MaterialAttributes, B=@n99, Alpha=@n100) {vertexAttributeBlendType=UseA}, False=@n1.MaterialAttributes) {group=Customization_Wrap, parameterName=Wrap_Flatline})
  NamedRerouteDeclaration(None=FunctionInput() {bUsePreviewValueAsDefault=true, inputName=isCharacter, inputType=FunctionInput_Scalar}) {name=IsCharacter, nodeColor=<rgba 0.3125 0 1 1>}
  n101 = FunctionInput() {bUsePreviewValueAsDefault=true, inputName=IsStatsB, inputType=FunctionInput_Vector2, sortPriority=78}
  NamedRerouteDeclaration(None=ComponentMask(None=NamedRerouteDeclaration(None=@n101) {name=Is_StatsB, nodeColor=<rgba 0 0.9253 1 1>}) {r=true}) {name=IsBackpack, nodeColor=<rgba 0.2656 0 1 1>}
 ]}
