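# Material function for the Flora cosmetic wrap, layering bark, moss and animated flowers over an incoming material within customization masks, behind the Wrap_Flora static switch.
# Projected bark gets height-blended moss from bounds gradients, normals and Gmap, with state inputs selecting per-item tuning.
# Flowers are randomized grid cells using flipbook variants, spin and pulsing scale.
# A Blueprint-set timestamp makes flowers appear, then shrink and fade over a set duration.
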MaterialFunction /Game/Weapons/WeaponSkins/Wraps/MaterialFunctions/MF_Wrap_Flora
graph {nodes [
  n1 = TextureObjectParameter() {group=Wrap__FLORA_Textures, parameterName=FLORA_BaseColor_01, sortPriority=11, texture=&T_BlacksRGB}
  n2 = VectorParameter() {group=Wrap_FLORA_Bark, parameterName=FLORA_Offset_Weapon_01---------------------, sortPriority=10, defaultValue=<rgba 0 0 0 1>}
  n3 = VectorParameter() {group=Wrap_FLORA_Bark, parameterName=FLORA_Offset_Vehicle_01, sortPriority=11, defaultValue=<rgba 0 0 0 1>}
  n4 = VectorParameter() {group=Wrap_FLORA_Bark, parameterName=FLORA_Offset_Vehicle_Side_01, sortPriority=11, defaultValue=<rgba 0 0 0 1>}
  n5 = FunctionInput(Preview=Constant()) {inputName=isState, inputType=FunctionInput_Vector4, sortPriority=5}
  n6 = MaterialFunctionCall(Color=@n5) {materialFunction=&SplitComponents}
  n7 = VectorParameter() {group=Wrap_FLORA_Bark, parameterName=FLORA_Offset_Mech_01, sortPriority=12, defaultValue=<rgba 0 0 0 1>}
  n8 = LinearInterpolate(A=LinearInterpolate(A=@n2.RGB, B=LinearInterpolate(A=@n3.RGB, B=@n4.RGB, Alpha=@n6.G), Alpha=@n6.R), B=@n7.RGB, Alpha=@n6.B)
  n9 = VertexInterpolator(VS=PreSkinnedPosition())
  n10 = PreSkinnedLocalBounds()
  n11 = FunctionInput(Preview=Divide(A=Subtract(A=@n9.PS, B=@n10.Min), B=Subtract(A=@n10.Max, B=@n10.Max))) {bUsePreviewValueAsDefault=true, inputName=Calcated_Bounds, sortPriority=11}
  n12 = VectorParameter() {group=Wrap_FLORA_Bark, parameterName="FLORA_Scale_Weapon_01 A=uniform-----", sortPriority=30, defaultValue=<rgba 1 1 1 1>}
  n13 = VectorParameter() {group=Wrap_FLORA_Bark, parameterName="FLORA_Scale_Vehicle_01 A=uniform", sortPriority=31, defaultValue=<rgba 1 1 1 1>}
  n14 = VectorParameter() {group=Wrap_FLORA_Bark, parameterName="FLORA_Scale_Vehicle_Side_01 A=uniform", sortPriority=31, defaultValue=<rgba 1 1 1 1>}
  n15 = LinearInterpolate(
   A=Multiply(A=@n12.RGB, B=@n12.RGB)
   B=LinearInterpolate(A=Multiply(A=@n13.RGB, B=@n13.RGB), B=Multiply(A=@n14.RGB, B=@n14.RGB), Alpha=@n6.G)
   Alpha=@n6.R
  )
  n16 = VectorParameter() {group=Wrap_FLORA_Bark, parameterName="FLORA_Scale_Mech_01 A=uniform", defaultValue=<rgba 1 1 1 1>}
  n17 = Multiply(A=@n16.RGB, B=@n16.RGB)
  n18 = VectorParameter() {group=Wrap_FLORA_Bark, parameterName="FLORA_Fill_alternate_Axis_Color_A=Roughness 01", sortPriority=0}
  n19 = VectorParameter() {group=Wrap_FLORA_Bark, parameterName=FLORA_Rotation-Weapon-01---------------------, sortPriority=1, defaultValue=<rgba 0 0 -0.25 1>}
  n20 = VectorParameter() {group=Wrap_FLORA_Bark, parameterName=FLORA_Rotation_Vehicle_01, sortPriority=2, defaultValue=<rgba 0 0 0 1>}
  n21 = VectorParameter() {group=Wrap_FLORA_Bark, parameterName=FLORA_Rotation_Vehicle_Side_01, sortPriority=2, defaultValue=<rgba 0 0 0 1>}
  n22 = LinearInterpolate(A=@n20.RGB, B=@n21.RGB, Alpha=@n6.G)
  n23 = VectorParameter() {group=Wrap_FLORA_Bark, parameterName=FLORA_Rotation_Mech_01, sortPriority=3, defaultValue=<rgba 0.25 0.25 0 1>}
  n24 = FunctionInput(Preview=VertexInterpolator(VS=PreSkinnedNormal()).PS) {bUsePreviewValueAsDefault=true, inputName=Pre-Skinned Local Normal, sortPriority=10}
  n25 = MaterialFunctionCall!(
   IsCharacter
   IsStatsB
   Texture=@n1
   Rotation_Axis=LinearInterpolate(A=LinearInterpolate(A=@n19.RGB, B=@n22, Alpha=@n6.R), B=@n23.RGB, Alpha=@n6.B)
   Caculated_Bounds=Multiply(A=Add(A=@n11, B=@n8), B=LinearInterpolate(A=@n15, B=@n17, Alpha=@n6.B))
   Pre_Skinned_Normals=@n24
   Blend_Contrast=LinearInterpolate(
    A=ScalarParameter() {defaultValue=5, group=Wrap_FLORA_Bark, parameterName=FLORA_Blend_Contrast, sortPriority=0}
    B=ScalarParameter() {defaultValue=2, group=Wrap_FLORA_Bark, parameterName=FLORA_Blend_Contrast_Mech, sortPriority=0}
    Alpha=@n6.B
   )
   Fill_alternate_Axis_Opacity=ScalarParameter() {group=Wrap_FLORA_Bark, parameterName=FLORA_Fill_alternate_Axis_Opacity 01, sortPriority=0}
   "Fill_alternate_Axis_Color_A=Roughness"=AppendVector(A=@n18.RGB, B=@n18.RGB)
   Is_Vehicle=@n6.R
   Is_Mech=@n6.B
   Enable_Fill_Z_Weapon
   Fill_Z_Weapon_opacity
   Fill_Z_Weapon_Color
  ) {materialFunction=&/MF_TopsSideProjection_Base}
  n26 = ChannelMaskParameter(None=@n25.Combined) {group=Wrap_FLORA_Bark, maskChannel=Alpha, parameterName=FLORA_Roughness_ChannelMask_01, sortPriority=-101, defaultValue=<rgba 0 0 0 1>}
  n27 = GetMaterialAttributes(None=FunctionInput() {inputName=Material, inputType=FunctionInput_MaterialAttributes, sortPriority=-20}) {attributeGetTypes [#1]}
  n28 = SetMaterialAttributes!(
   MaterialAttributes=@n27.MaterialAttributes
   BaseColor=ComponentMask(None=@n25.Combined) {b=true, g=true, r=true}
   Metallic=Constant()
   Specular=LinearInterpolate(
    A=ScalarParameter() {defaultValue=0.35, group=Wrap_FLORA_Bark, parameterName=FLORA_Specular_Min, sortPriority=-100}
    B=ScalarParameter() {defaultValue=0.65, group=Wrap_FLORA_Bark, parameterName=FLORA_Specular_Max, sortPriority=-100}
    Alpha=OneMinus(None=@n26.R)
   )
   Roughness=LinearInterpolate(
    A=ScalarParameter() {group=Wrap_FLORA_Bark, parameterName=FLORA_Roughness, sortPriority=-100}
    B=ScalarParameter() {defaultValue=1, group=Wrap_FLORA_Bark, parameterName=FLORA_Roughness_High, sortPriority=-100}
    Alpha=@n26.R
   )
  ) {attributeSetTypes [
    #2
    #3
    #4
    #5
   ]}
  n29 = Constant() {r=1}
  n30 = FunctionInput(Preview=@n29) {
   bUsePreviewValueAsDefault=true
   desc=secondary
   inputName="Customization Mask 2 (_CM B Top Right)"
   inputType=FunctionInput_Scalar
   sortPriority=50
  }
  n31 = FunctionInput(Preview=@n29) {
   bUsePreviewValueAsDefault=true
   desc=base customization
   inputName="Customization Mask (_CM B Top left)"
   inputType=FunctionInput_Scalar
   sortPriority=49
  }
  n32 = Add(A=@n30, B=@n31)
  n33 = BlendMaterialAttributes(
   A=@n27.MaterialAttributes
   B=@n28
   Alpha=LinearInterpolate(A=Saturate(None=@n32), B=NamedRerouteDeclaration(None=@n30) {name=PrimaryMask, nodeColor=<rgba 0 0.8828 1 1>}, Alpha=NamedRerouteUsage())
  ) {vertexAttributeBlendType=UseA}
  n34 = GetMaterialAttributes(None=@n33) {attributeGetTypes [#2, #4, #5]}
  n35 = MaterialFunctionCall(Color=@n5) {materialFunction=&SplitComponents}
  n36 = LinearInterpolate(
   A=ScalarParameter() {defaultValue=16, group=Wrap_FLORA_Flower, parameterName=FLORA_Flower_TileRate_Weapon, sortPriority=100}
   B=LinearInterpolate(
    A=ScalarParameter() {defaultValue=16, group=Wrap_FLORA_Flower, parameterName=FLORA_Flower_TileRate_Vehicle, sortPriority=100}
    B=ScalarParameter() {defaultValue=16, group=Wrap_FLORA_Flower, parameterName=FLORA_Flower_TileRate_VehicleSide, sortPriority=100}
    Alpha=@n35.G
   )
   Alpha=@n35.R
  )
  n37 = ScalarParameter() {defaultValue=16, group=Wrap_FLORA_Flower, parameterName=FLORA_Flower_TileRate_Mech-----, sortPriority=100}
  n38 = LinearInterpolate(
   A=LinearInterpolate(
    A=LinearInterpolate(A=@n36, B=@n37, Alpha=@n35.B)
    B=ScalarParameter() {defaultValue=16, group=Wrap_FLORA_Flower, parameterName=FLORA_Flower_TileRate_Backpack-----, sortPriority=100}
    Alpha=NamedRerouteUsage()
   )
   B=ScalarParameter() {defaultValue=16, group=Wrap_FLORA_Flower, parameterName=FLORA_Flower_TileRate_Pickaxe-----, sortPriority=100}
   Alpha=NamedRerouteUsage()
  )
  n39 = MaterialFunctionCall(Color=@n5) {materialFunction=&SplitComponents}
  n40 = MaterialFunctionCall!(
   For_Safe_Zones
   Normal_Map
   Caculated_Bounds=@n11
   Pre_Skinned_Normals=@n24
   Is Vehicle=@n39.R
   Is VehicleSide=@n39.R
   Is Mech=@n39.R
   Vehicle_Rotation
   UV_Rotation Bypass
   Mech_Rotation
   Uvs_Offset
   Offset-after rotation
   Uvs_TileRate
  ) {materialFunction=&/MF_Projection_Side_Normal}
  n41 = ComponentMask(None=@n40.UVs No Offsets/Tiling) {r=true}
  n42 = Divide(A=Floor(None=Multiply(A=@n41, B=@n38) {constB=5}), B=@n38) {constB=5}
  n43 = ScalarParameter() {defaultValue=4, group=Wrap_FLORA_Flower, parameterName=FLORA_Flower_Offset_Y-----, sortPriority=40}
  n44 = Add(A=Multiply(A=@n40.UVs No Offsets/Tiling, B=@n38), B=AppendVector(A=Constant(), B=Multiply(A=@n42, B=@n43))) {constB=0}
  n45 = VectorParameter() {group=Wrap_FLORA_Flower, parameterName=FLORA_Flower_Random_seed, sortPriority=40, defaultValue=<rgba 4 3 0 0>}
  n46 = MaterialFunctionCall!(2 Vector=Add(A=Floor(None=@n44), B=ComponentMask(None=@n45.RGB) {g=true, r=true}), RandomSeed) {materialFunction=&/MF_Random_2}
  n47 = LinearInterpolate(Alpha=@n46.utput) {constA=-2, constB=2}
  n48 = Constant() {r=2}
  n49 = LinearInterpolate(
   A=ScalarParameter() {defaultValue=0.5, group=Wrap_FLORA_Flower, parameterName=FLORA_Flower_Scale-----, sortPriority=10}
   B=LinearInterpolate(
    A=ScalarParameter() {defaultValue=0.5, group=Wrap_FLORA_Flower, parameterName=FLORA_Flower_Scale_Character-----, sortPriority=10}
    B=ScalarParameter() {defaultValue=0.5, group=Wrap_FLORA_Flower, parameterName=FLORA_Flower_Scale_Backpack-----, sortPriority=10}
    Alpha=NamedRerouteUsage()
   )
   Alpha=NamedRerouteUsage()
  )
  n50 = Time()
  n51 = Multiply(A=ScalarParameter() {defaultValue=0.7, group=Wrap_FLORA_Flower, parameterName=FLORA_Flower_Scale_Speed, sortPriority=10}, B=@n46.utput)
  n52 = Multiply(
   A=Sine(None=Multiply(A=@n50, B=@n51))
   B=ScalarParameter() {defaultValue=0.35, group=Wrap_FLORA_Flower, parameterName=FLORA_Flower_Scale_Sine_Amount, sortPriority=10}
  )
  n53 = MaterialFunctionCall(Color=@n5) {materialFunction=&SplitComponents}
  n54 = LinearInterpolate(
   A=ScalarParameter() {defaultValue=-0.4, group=Wrap_FLORA_Moss, parameterName=FLORA_MossBuildup_Grad_offset_Weapon Z, sortPriority=100}
   B=ScalarParameter() {defaultValue=-0.4, group=Wrap_FLORA_Moss, parameterName=FLORA_MossBuildup_Grad_offset_Vehicle Z, sortPriority=100}
   Alpha=@n53.R
  )
  n55 = ScalarParameter() {defaultValue=-0.4, group=Wrap_FLORA_Moss, parameterName=FLORA_MossBuildup_Grad_offset_Mech Z, sortPriority=100}
  n56 = ScalarParameter() {defaultValue=0.5, group=Wrap_FLORA_Moss, parameterName=FLORA_MossBuildup_Grad_Falloff_Weapon Z, sortPriority=100}
  n57 = ScalarParameter() {defaultValue=0.5, group=Wrap_FLORA_Moss, parameterName=FLORA_MossBuildup_Grad_Falloff_Vehicle Z, sortPriority=100}
  n58 = Divide(
   A=Add(A=@n11, B=LinearInterpolate(A=@n54, B=@n55, Alpha=@n53.B))
   B=LinearInterpolate(
    A=LinearInterpolate(A=@n56, B=@n57, Alpha=@n53.R)
    B=ScalarParameter() {defaultValue=0.5, group=Wrap_FLORA_Moss, parameterName=FLORA_MossBuildup_Grad_Falloff_Mech Z, sortPriority=100}
    Alpha=@n53.B
   )
  )
  n59 = ScalarParameter() {defaultValue=1, group=Wrap_FLORA_Moss, parameterName=FLORA_ForRoll_OverrideGradient, sortPriority=-100}
  n60 = LinearInterpolate(A=ComponentMask(None=@n58) {r=true}, B=ComponentMask(None=@n58) {b=true}, Alpha=@n59)
  n61 = MaterialFunctionCall(Color=Abs(None=@n24)) {materialFunction=&SplitComponents}
  n62 = MaterialFunctionCall(Color=@n5) {materialFunction=&SplitComponents}
  n63 = LinearInterpolate(A=LinearInterpolate(A=Add(A=@n61.R, B=@n61.R), B=Add(A=@n61.B, B=@n61.B), Alpha=@n62.R), B=Add(A=@n61.R, B=@n61.R), Alpha=@n62.B)
  n64 = ScalarParameter() {defaultValue=1.75, group=Wrap_FLORA_Moss, parameterName=FLORA_MossBuildup_GradMask_Falloff_Weapon, sortPriority=100}
  n65 = Saturate(None=MaterialFunctionCall(
    In=Divide(A=@n63, B=@n64)
    Contrast=ScalarParameter() {defaultValue=5, group=Wrap_FLORA_Moss, parameterName=FLORA_MossBuildup_GradMask_Contrast_Weapon, sortPriority=100}
   ) {materialFunction=&CheapContrast}.Result)
  n66 = StaticBool() {value=true}
  n67 = TextureObjectParameter() {
   group=Wrap__FLORA_Textures
   parameterName=FLORA_Mask_03
   sortPriority=50
   texture=/Game/Weapons/WeaponSkins/Wraps/Textures/Flora/T_Wrap_Flora_Flower_M1
  }
  n68 = Constant() {r=1}
  n69 = ScalarParameter() {defaultValue=1, group=Wrap_FLORA_Moss, parameterName=FLORA_Moss_TileRate_Weapon}
  n70 = MaterialFunctionCall(Color=@n5) {materialFunction=&SplitComponents}
  n71 = LinearInterpolate(
   A=ScalarParameter() {defaultValue=1, group=Wrap_FLORA_Moss, parameterName=FLORA_Moss_TileRate_Vehicle}
   B=ScalarParameter() {defaultValue=1, group=Wrap_FLORA_Moss, parameterName=FLORA_Moss_TileRate_VehicleSide}
   Alpha=@n70.G
  )
  n72 = Multiply(
   A=LinearInterpolate(
    A=LinearInterpolate(A=@n69, B=@n71, Alpha=@n70.R)
    B=ScalarParameter() {defaultValue=1, group=Wrap_FLORA_Moss, parameterName=FLORA_Moss_TileRate_Mech -----}
    Alpha=@n70.B
   )
   B=@n11
  )
  n73 = MaterialFunctionCall(Color=Constant3Vector() {constant=<rgba 0 -0.25 -0.25 1>}) {materialFunction=&SplitComponents}
  n74 = ScalarParameter() {defaultValue=4, group=Wrap_FLORA_Moss, parameterName=FLORA_Moss_Blend_Contrast, sortPriority=1}
  n75 = MaterialFunctionCall(
   Export Float 4=@n66
   TextureObject=@n67
   TextureSize=@n68
   WorldPosition=@n72
   World Space Normal=@n24
   RotZAxis=@n73.B
   RotXAxis=@n73.B
   RotYAxis=@n73.B
   ProjectionTransitionContrast=@n74
  ) {materialFunction=&/WorldAlignedTexture}
  n76 = Add(
   A=Saturate(None=Multiply(A=@n60, B=@n65))
   B=Multiply(
    A=ComponentMask(None=@n75.XYZ Texture) {b=true}
    B=ScalarParameter() {defaultValue=0.25, group=Wrap_FLORA_Moss, parameterName=FLORA_MossBuildup_Grad_Warp Amount, sortPriority=100}
   ) {constB=0.1}
  )
  n77 = Subtract(
   A=Time()
   B=ScalarParameter() {defaultValue=-9999999, group=Wrap_FLORA_Reactive, parameterName=TimeFrom-BP----------------, sortPriority=-600}
  )
  n78 = ScalarParameter() {defaultValue=1.5, group=Wrap_FLORA_Reactive, parameterName=FLORA_FLower_AnimationDuration, sortPriority=33}
  n79 = Power(
   Base=Saturate(None=Divide(A=@n77, B=@n78))
   Exp=ScalarParameter() {defaultValue=0.85, group=Wrap_FLORA_Reactive, parameterName=FLORA_Flower_CurveAmount, sortPriority=34}
  )
  n80 = ScalarParameter() {group=Wrap_FLORA_Reactive, parameterName=FLORA_TestReactive, sortPriority=300}
  n81 = Multiply(
   A=Multiply(A=Add(A=@n49, B=@n52) {constA=0.8}, B=@n76)
   B=LinearInterpolate(A=OneMinus(None=@n79) {desc=inverts curve from 0-1 to 1-0}, B=@n80, Alpha=@n80)
  )
  n82 = Time()
  n83 = LinearInterpolate(
   A=ScalarParameter() {defaultValue=-0.25, group=Wrap_FLORA_Flower, parameterName=FLORA_Flower_Rotation_Speed_Range_A, sortPriority=25}
   B=ScalarParameter() {defaultValue=0.25, group=Wrap_FLORA_Flower, parameterName=FLORA_Flower_Rotation_Speed_Range_B, sortPriority=25}
   Alpha=@n46.utput
  )
  n84 = MaterialFunctionCall(
   UVs=MaterialFunctionCall(UVs=Frac(None=@n44), Texture Scale=Clamp(None=@n81)) {materialFunction=&ScaleUVsByCenter}.UVs
   "Rotation Angle (0-1)"=Add(A=Multiply(A=@n82, B=@n83), B=@n46.utput)
  ) {materialFunction=&CustomRotator}
  n85 = TextureObjectParameter() {
   group=Wrap__FLORA_Textures
   parameterName=FLORA_BaseColor_03
   sortPriority=50
   texture=/Game/Weapons/WeaponSkins/Wraps/Textures/Flora/T_Wrap_Flora_Flower_BC
  }
  n86 = MaterialFunctionCall("Animation  Phase (0-1)"=@n47, Number of Rows=@n48, Number of Columns=@n48, Texture=@n85, UVs=@n84.Rotated Values) {materialFunction=&FlipBook}
  n87 = MaterialFunctionCall(Color=@n86.Result) {materialFunction=&SplitComponents}
  n88 = Constant()
  n89 = ScalarParameter() {defaultValue=0.1, group=Wrap_FLORA_Flower, parameterName=FLORA_FlowerMask_Contrast, sortPriority=200}
  n90 = Saturate(None=OneMinus(None=MaterialFunctionCall(In=@n63, Contrast=@n89) {materialFunction=&CheapContrast}.Result))
  n91 = ScalarParameter() {defaultValue=1, group=Wrap_FLORA_Flower, parameterName=FLORA_FlowerMask_AltAxisActive -----, sortPriority=200}
  n92 = Constant()
  n93 = MaterialFunctionCall(Color=Abs(None=@n24)) {materialFunction=&SplitComponents}
  n94 = LinearInterpolate(A=LinearInterpolate(A=@n93.R, B=LinearInterpolate(A=@n93.B, B=@n93.B, Alpha=@n62.G), Alpha=@n62.R), B=@n93.G, Alpha=@n62.B)
  n95 = Saturate(None=OneMinus(None=MaterialFunctionCall(In=@n94, Contrast=@n89) {materialFunction=&CheapContrast}.Result))
  n96 = ScalarParameter() {defaultValue=1, group=Wrap_FLORA_Flower, parameterName=FLORA_FlowerMask_Z+AltAxisActive, sortPriority=200}
  n97 = OneMinus(None=LinearInterpolate(A=LinearInterpolate(A=@n88, B=@n90, Alpha=@n91), B=LinearInterpolate(A=@n92, B=@n95, Alpha=@n96), Alpha=@n96))
  n98 = ScalarParameter() {defaultValue=2, group=Wrap_FLORA_Moss, parameterName=FLORA_MossBuildup_Z_Strength_Weapon, sortPriority=100}
  n99 = ScalarParameter() {defaultValue=0.7, group=Wrap_FLORA_Moss, parameterName=FLORA_MossBuildup_Gmap_influence Z-----, sortPriority=100}
  n100 = ScalarParameter() {defaultValue=0.96, group=Wrap_FLORA_Moss, parameterName=FLORA_MossBuildup_Gmap_influence Z_BackPack-----, sortPriority=100}
  n101 = ComponentMask(None=NamedRerouteUsage()) {r=true}
  n102 = TextureSample() {samplerType=SAMPLERTYPE_Masks, texture=/Game/UI/Wrapping_Paper/Textures/T_Wrapping_Paper_CM}
  n103 = Multiply(
   A=Subtract(
    A=FunctionInput(Preview=@n102.A) {bUsePreviewValueAsDefault=true, inputName=Gmap, inputType=FunctionInput_Scalar, sortPriority=24}
    B=LinearInterpolate(A=@n99, B=@n100, Alpha=@n101)
   ) {constA=0.5}
   B=ScalarParameter() {defaultValue=10, group=Wrap_FLORA_Moss, parameterName=FLORA_MossBuildup_Gmap_Strength Z, sortPriority=100}
  )
  n104 = Saturate(None=Multiply(
    A=ComponentMask(None=@n24) {b=true}
    B=ScalarParameter() {defaultValue=8, group=Wrap_FLORA_Moss, parameterName=FLORA_MossBuildup_LNormal_Strength Z, sortPriority=100}
   ))
  n105 = MaterialFunctionCall!(PivotOffset) {materialFunction=&/MF_SphereProjectedNormals}
  n106 = Constant3Vector() {constant=<rgba 0 0 1 0>}
  n107 = Saturate(None=LinearInterpolate(A=Add(A=@n103, B=@n104), B=DotProduct(A=@n105.Result, B=@n106), Alpha=NamedRerouteUsage()))
  n108 = ScalarParameter() {defaultValue=1.5, group=Wrap_FLORA_Moss, parameterName=FLORA_MossBuildup_Z_Strength_Vehicle, sortPriority=100}
  n109 = Multiply(A=LinearInterpolate(A=Multiply(A=@n107, B=@n98), B=Multiply(A=@n107, B=@n108), Alpha=@n53.R), B=@n59)
  n110 = Saturate(None=Add(A=@n76, B=Saturate(None=@n109)))
  n111 = Constant()
  n112 = Multiply(
   A=SphereMask(A=@n84.Rotated Values, B=Constant() {r=0.5}, Radius=Constant() {r=0.5}, Hardness=Constant() {r=0.9})
   B=Round(None=MaterialFunctionCall(Transition Phase=@n110, Height Texture=@n46.utput, Contrast=@n111) {materialFunction=&HeightLerp}.Alpha)
  )
  n113 = Multiply(A=@n112, B=LinearInterpolate(B=@n86.Alpha, Alpha=@n97))
  n114 = TextureObjectParameter() {group=Wrap__FLORA_Textures, parameterName=FLORA_BaseColor_02, texture=/Game/Weapons/WeaponSkins/Wraps/Textures/Flora/T_Wrap_Flora_Moss_BC}
  n115 = MaterialFunctionCall(
   Export Float 4=@n66
   TextureObject=@n114
   TextureSize=@n68
   WorldPosition=@n72
   World Space Normal=@n24
   RotZAxis=@n73.R
   RotXAxis=@n73.R
   RotYAxis=@n73.R
   ProjectionTransitionContrast=@n74
  ) {materialFunction=&WorldAlignedTexture}
  n116 = MaterialFunctionCall(
   Transition Phase=@n110
   Height Texture=ComponentMask(None=@n75.XYZ Texture) {a=true}
   Contrast=ScalarParameter() {defaultValue=10, group=Wrap_FLORA_Moss, parameterName=FLORA_MossBuildup_Z_Contrast, sortPriority=100}
  ) {materialFunction=&HeightLerp}
  n117 = LinearInterpolate(A=@n34.BaseColor, B=ComponentMask(None=@n115.XYZ Texture) {b=true, g=true, r=true}, Alpha=@n116.Alpha)
  n118 = Saturate(None=Divide(A=@n77, B=ScalarParameter() {defaultValue=1.5, group=Wrap_FLORA_Reactive, parameterName=FLORA_AnimationDuration, sortPriority=33}))
  n119 = ScalarParameter() {defaultValue=0.85, group=Wrap_FLORA_Reactive, parameterName=FLORA_Curve_Amount, sortPriority=34}
  n120 = LinearInterpolate(A=OneMinus(None=Power(Base=@n118, Exp=@n119)) {desc=inverts curve from 0-1 to 1-0}, B=@n80, Alpha=@n80)
  n121 = MaterialFunctionCall(Color=MaterialFunctionCall("Animation  Phase (0-1)"=@n47, Number of Rows=@n48, Number of Columns=@n48, Texture=@n67, UVs=@n84.Rotated Values) {materialFunction=&FlipBook}.Result) {materialFunction=&SplitComponents}
  n122 = Multiply(A=@n112, B=LinearInterpolate(B=Multiply(A=@n86.Alpha, B=@n121.R), Alpha=@n97))
  n123 = LinearInterpolate(
   A=ScalarParameter() {defaultValue=0.25, group=Wrap_FLORA_Flower, parameterName=FLORA_Flower_Specular, sortPriority=-100}
   B=ScalarParameter() {defaultValue=0.5, group=Wrap_FLORA_Flower, parameterName=FLORA_Flower_SpecularHigh, sortPriority=-100}
   Alpha=OneMinus(None=@n122)
  )
  n124 = ComponentMask(None=@n115.XYZ Texture) {a=true}
  n125 = LinearInterpolate(
   A=@n34.Specular
   B=LinearInterpolate(
    A=ScalarParameter() {group=Wrap_FLORA_Moss, parameterName=FLORA_Moss_Specular_Min, sortPriority=-100}
    B=ScalarParameter() {defaultValue=0.55, group=Wrap_FLORA_Moss, parameterName=FLORA_Moss_Specular_Max, sortPriority=-100}
    Alpha=OneMinus(None=@n124)
   )
   Alpha=@n116.Alpha
  )
  n126 = LinearInterpolate(
   A=ScalarParameter() {group=Wrap_FLORA_Flower, parameterName=FLORA_Flower_Roughness, sortPriority=-100}
   B=ScalarParameter() {defaultValue=1, group=Wrap_FLORA_Flower, parameterName=FLORA_Flower_RoughnessHigh, sortPriority=-100}
   Alpha=@n122
  )
  n127 = LinearInterpolate(
   A=@n34.Roughness
   B=LinearInterpolate(
    A=ScalarParameter() {group=Wrap_FLORA_Moss, parameterName=FLORA_Moss_Roughness_Min, sortPriority=-100}
    B=ScalarParameter() {defaultValue=0.55, group=Wrap_FLORA_Moss, parameterName=FLORA_Moss_Roughness_Max, sortPriority=-100}
    Alpha=@n124
   )
   Alpha=@n116.Alpha
  )
  n128 = SetMaterialAttributes!(
   MaterialAttributes=@n34.MaterialAttributes
   BaseColor=LinearInterpolate(A=@n117, B=LinearInterpolate(A=@n117, B=@n87.RGB, Alpha=@n113), Alpha=@n120)
   Specular=LinearInterpolate(A=@n125, B=LinearInterpolate(A=@n125, B=@n123, Alpha=@n113), Alpha=@n120)
   Roughness=LinearInterpolate(A=@n127, B=LinearInterpolate(A=@n127, B=@n126, Alpha=@n113), Alpha=@n120)
  ) {attributeSetTypes [#2, #4, #5]}
  n129 = LinearInterpolate(B=NamedRerouteUsage(), Alpha=NamedRerouteUsage()) {constA=1}
  FunctionOutput(None=StaticSwitchParameter(True=BlendMaterialAttributes(A=@n33, B=@n128, Alpha=@n129) {vertexAttributeBlendType=UseA}, False=@n27.MaterialAttributes) {group=Customization_Wrap, parameterName=Wrap_Flora})
  NamedRerouteDeclaration(None=MaterialFunctionCall(Float4=@n5) {materialFunction=&BreakOutFloat4Components}.A) {name=IsCharacter, nodeColor=<rgba 0.9033 1 0 1>}
  n130 = FunctionInput() {bUsePreviewValueAsDefault=true, inputName=IsStatsB, inputType=FunctionInput_Vector2, sortPriority=78}
  NamedRerouteDeclaration(None=@n130) {name=Is_StatsB, nodeColor=<rgba 0 0.9253 1 1>}
  NamedRerouteDeclaration(None=ComponentMask(None=@n130) {r=true}) {name=IsBackpack, nodeColor=<rgba 0.1754 1 0.9502 1>}
  NamedRerouteDeclaration(None=ComponentMask(None=@n130) {g=true}) {name=IsPickaxe, nodeColor=<rgba 0.1562 1 1 1>}
 ]}
