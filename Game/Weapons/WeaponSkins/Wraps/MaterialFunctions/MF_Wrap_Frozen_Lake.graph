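# Frozen Lake weapon wrap material function that paints a faux-deep ice look onto weapons, vehicles, mechs and characters, with per-type tiling and depth parameters.
# Crack, dirt and fish textures are projected from pre-skinned bounds and offset along the local camera vector in stacked parallax steps, with switches to reduce or remove layers.
# Curve-atlas colors, a gradient background and curvature edge highlights form time-of-day-scaled emissive.
# The result, with surface normal and roughness, blends over the input material by customization mask.
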MaterialFunction /Game/Weapons/WeaponSkins/Wraps/MaterialFunctions/MF_Wrap_Frozen_Lake
graph {nodes [
  n1 = Transform(None=CameraVectorWS()) {transformSourceType=TRANSFORMSOURCE_World, transformType=TRANSFORM_Local}
  n2 = MaterialFunctionCall() {materialFunction=/Engine/Functions/Engine_MaterialFunctions02/WorldPositionOffset/ObjectScale}
  n3 = Constant()
  n4 = FunctionInput(Preview=@n3) {inputName=isVehicle, inputType=FunctionInput_Scalar, sortPriority=5}
  n5 = FunctionInput(Preview=@n3) {bUsePreviewValueAsDefault=true, inputName=isMech, inputType=FunctionInput_Scalar, sortPriority=4}
  n6 = FunctionInput() {bUsePreviewValueAsDefault=true, inputName=isCharacter, inputType=FunctionInput_Scalar, sortPriority=4}
  n7 = MaterialFunctionCall(
   X=@n4
   Y=FunctionInput(Preview=@n3) {bUsePreviewValueAsDefault=true, inputName=Side Project Vehicle, inputType=FunctionInput_Scalar}
   Z=@n5
   A=@n6
  ) {materialFunction=&MakeFloat4}
  n8 = MaterialFunctionCall(Float4=@n7.Result) {materialFunction=&BreakOutFloat4Components}
  n9 = LinearInterpolate(
   A=ScalarParameter() {defaultValue=4.1181, group=Wrap_FrozenLake, parameterName=FL_Layers_Depth_Amount}
   B=ScalarParameter() {defaultValue=4.1181, group=Wrap_FrozenLake, parameterName=FL_Layers_Depth_Amount_Vehicle}
   Alpha=@n8.R
  )
  n10 = ScalarParameter() {defaultValue=4.1181, group=Wrap_FrozenLake, parameterName=FL_Layers_Depth_Amount_Mech}
  n11 = Multiply(A=@n1, B=Multiply(A=Saturate(None=@n2.Scale XYZ), B=LinearInterpolate(A=@n9, B=@n10, Alpha=@n8.B)))
  n12 = FunctionInput(Preview=TextureSample() {samplerType=SAMPLERTYPE_Masks, texture=/Game/Weapons/FORT_Rifles/Textures/T_Scar_CM}.G) {bUsePreviewValueAsDefault=true, inputName=Curavture, inputType=FunctionInput_Scalar, sortPriority=24}
  n13 = PreSkinnedPosition()
  n14 = PreSkinnedLocalBounds()
  n15 = Divide(A=Subtract(A=VertexInterpolator(VS=@n13).PS, B=@n14.Min), B=Subtract(A=@n14.Max, B=@n14.Max))
  n16 = FunctionInput(Preview=@n15) {bUsePreviewValueAsDefault=true, inputName=Calcated_Bounds, sortPriority=11}
  n17 = Add(A=@n16, B=Multiply(A=OneMinus(None=@n12), B=ScalarParameter() {group=Wrap_FrozenLake, parameterName=FL_Layers_Warp_Cur, sortPriority=1}))
  n18 = Add(A=@n17, B=Multiply(A=@n11) {constB=-0.011})
  n19 = LinearInterpolate(
   A=LinearInterpolate(
    A=ScalarParameter() {defaultValue=0.35, group=Wrap_FrozenLake, parameterName=FL_Layers_Skew_Amount}
    B=ScalarParameter() {defaultValue=0.35, group=Wrap_FrozenLake, parameterName=FL_Layers_Skew_Amount_Vehicle}
    Alpha=@n8.R
   )
   B=LinearInterpolate(
    A=ScalarParameter() {defaultValue=0.35, group=Wrap_FrozenLake, parameterName=FL_Layers_Skew_Amount_Mech}
    B=ScalarParameter() {defaultValue=0.35, group=Wrap_FrozenLake, parameterName=FL_Layers_Skew_Amount_Character}
    Alpha=@n8.A
   )
   Alpha=@n8.B
  )
  n20 = VectorParameter() {group=Wrap_FrozenLake, parameterName=FL_Layers_TileRate, defaultValue=<rgba 1 1.5 1 1>}
  n21 = VectorParameter() {group=Wrap_FrozenLake, parameterName=FL_Layers_TileRate_Vehicle, defaultValue=<rgba 1 1.5 1 1>}
  n22 = VectorParameter() {group=Wrap_FrozenLake, parameterName=FL_Layers_TileRate_Mech, defaultValue=<rgba 1 1.5 1 1>}
  n23 = VectorParameter() {group=Wrap_FrozenLake, parameterName=FL_Layers_TileRate_Character, defaultValue=<rgba 1 1 1 1>}
  n24 = LinearInterpolate(
   A=LinearInterpolate(A=Multiply(A=@n20.RGB, B=@n20.RGB), B=Multiply(A=@n21.RGB, B=@n21.RGB), Alpha=@n8.R)
   B=LinearInterpolate(A=Multiply(A=@n22.RGB, B=@n22.RGB), B=Multiply(A=@n23.RGB, B=@n23.RGB), Alpha=@n8.A)
   Alpha=@n8.B
  )
  n25 = TextureObjectParameter() {
   group=Wrap_FrozenLake
   parameterName="FL_A=Layer Crack GB=Dirt Layers"
   samplerType=SAMPLERTYPE_Masks
   sortPriority=10
   texture=&T_BlackMask
  }
  n26 = Multiply(A=TextureSample(
    UVs=MaterialFunctionCall!(
     Normal_Map
     Caculated_Bounds=@n18
     Pre_Skinned_Normals
     Skew Amount=@n19
     Is Vehicle=@n4
     Uvs_Offset
     Uvs_TileRate=@n24
     Alternate Axies contrast
    ) {materialFunction=&/MF_Projection_Skew_Normal}.UVs
    Tex=@n25
   ).A) {constB=0.08}
  n27 = MaterialFunctionCall(Caculated_Bounds=Add(A=@n17, B=Multiply(A=@n11) {constB=-0.012}), Skew Amount=@n19, Is Vehicle=@n4, Uvs_TileRate=@n24) {materialFunction=&MF_Projection_Skew_Normal}
  Add(A=@n26, B=Multiply(A=TextureSample(UVs=@n27.UVs, Tex=@n25).A) {constB=0.08})
  n28 = Add(A=@n17, B=Multiply(A=@n1, B=ScalarParameter() {defaultValue=-0.0667, group=Wrap_FrozenLake, parameterName=FL_Dirt_Layer_DepthA}))
  n29 = VectorParameter() {group=Wrap_FrozenLake, parameterName=FL_Dirt_layers Tile_Rate, defaultValue=<rgba 1 2 0 1>}
  n30 = VectorParameter() {group=Wrap_FrozenLake, parameterName=FL_Dirt_layers Tile_Rate_Vehicle, defaultValue=<rgba 1 2 0 1>}
  n31 = MaterialFunctionCall(Float4=@n7.Result) {materialFunction=&BreakOutFloat4Components}
  n32 = VectorParameter() {group=Wrap_FrozenLake, parameterName=FL_Dirt_layers Tile_Rate_Mech, defaultValue=<rgba 1 2 0 1>}
  n33 = VectorParameter() {group=Wrap_FrozenLake, parameterName=FL_Dirt_layers Tile_Rate_Character, defaultValue=<rgba 1 2 0 1>}
  n34 = LinearInterpolate(
   A=LinearInterpolate(A=Multiply(A=@n29.RGB, B=@n29.RGB), B=Multiply(A=@n30.RGB, B=@n30.RGB), Alpha=@n31.R)
   B=LinearInterpolate(A=Multiply(A=@n32.RGB, B=@n32.RGB), B=Multiply(A=@n33.RGB, B=@n33.RGB), Alpha=@n31.A)
   Alpha=@n31.B
  )
  n35 = FunctionInput(Preview=VertexInterpolator(VS=PreSkinnedNormal()).PS) {bUsePreviewValueAsDefault=true, inputName=Pre-Skinned Local Normal, sortPriority=10}
  n36 = Multiply(
   A=TextureSample(
    UVs=MaterialFunctionCall(Caculated_Bounds=@n28, Pre_Skinned_Normals=@n35, Skew Amount=@n19, Is Vehicle=@n4, Uvs_TileRate=@n34) {materialFunction=&MF_Projection_Skew_Normal}.UVs
    Tex=@n25
   ).G
   B=ScalarParameter() {defaultValue=0.3, group=Wrap_FrozenLake, parameterName=FL_Dirt_Layer_BrightnessA}
  ) {constB=0.16}
  n37 = Multiply(A=@n1, B=ScalarParameter() {defaultValue=-0.181, group=Wrap_FrozenLake, parameterName=FL_Dirt_Layer_DepthB})
  n38 = TextureSample(
   UVs=MaterialFunctionCall(Caculated_Bounds=Add(A=@n17, B=@n37), Skew Amount=@n19, Is Vehicle=@n4, Uvs_TileRate=@n34) {materialFunction=&MF_Projection_Skew_Normal}.UVs
   Tex=TextureObjectParameter() {group=Wrap_FrozenLake, parameterName=FL__Layer__Fish, texture=&T_BlacksRGB}
  )
  Add(
   A=@n36
   B=Multiply(A=@n38.B, B=ScalarParameter() {defaultValue=0.15, group=Wrap_FrozenLake, parameterName=FL_Dirt_Layer_BrightnessB}) {constB=0.08}
  )
  Constant() {r=1}
  n39 = FunctionInput(Preview=SetMaterialAttributes!(MaterialAttributes, Normal=TextureSample() {samplerType=SAMPLERTYPE_Normal, texture=/Game/Weapons/FORT_Rifles/T_SCAR_N}.RGB) {attributeSetTypes [#1]}) {inputName=Material, inputType=FunctionInput_MaterialAttributes, sortPriority=-20}
  n40 = GetMaterialAttributes(None=@n39)
  n41 = VectorParameter() {group=Wrap_FrozenLake, parameterName=FL_BS_Color, sortPriority=1, defaultValue=<rgba 0.003 0.003 0.003 1>}
  n42 = ScalarParameter() {defaultValue=0.15, group=Wrap_FrozenLake, parameterName=FL_BS_Roughness_Low}
  n43 = ScalarParameter() {defaultValue=0.5, group=Wrap_FrozenLake, parameterName=FL_BS_Roughness_high}
  n44 = TextureObjectParameter() {
   group=Wrap_FrozenLake
   parameterName=FL_Surface_Normals
   samplerType=SAMPLERTYPE_Normal
   sortPriority=10
   texture=&FlatNormalBlue
  }
  n45 = MaterialFunctionCall(
   Normal_Map=@n44
   Caculated_Bounds=Add(A=@n17, B=Multiply(A=@n11) {constB=-0.001})
   Pre_Skinned_Normals=@n35
   Skew Amount=@n19
   Is Vehicle=@n4
   Uvs_TileRate=@n24
   Alternate Axies contrast=ScalarParameter() {defaultValue=10, group=Wrap_FrozenLake, parameterName=FL_Alternate_Axies_Contrast, sortPriority=1}
  ) {materialFunction=&MF_Projection_Skew_Normal}
  n46 = TextureObjectParameter() {
   group=Wrap_FrozenLake
   parameterName="FL_R=Layer Mask_A=Roughness"
   samplerType=SAMPLERTYPE_Masks
   sortPriority=10
   texture=&T_BlackMask
  }
  n47 = TextureSample(UVs=@n45.UVs, Tex=@n46) {texture=/Game/Weapons/WeaponSkins/Wraps/Textures/FrozenLake/T_Wrap_Frozen_Mask_1}
  n48 = Multiply(A=TextureSample(UVs=@n45.UVs, Tex=@n25).R)
  n49 = Add(A=Multiply(A=@n11) {constB=-0.002}, B=@n17)
  n50 = Multiply(A=TextureSample(
    UVs=MaterialFunctionCall(Caculated_Bounds=@n49, Skew Amount=@n19, Is Vehicle=@n4, Uvs_TileRate=@n24) {materialFunction=&MF_Projection_Skew_Normal}.UVs
    Tex=@n25
   ).A) {constB=0.96}
  n51 = MaterialFunctionCall(Caculated_Bounds=Add(A=@n17, B=Multiply(A=@n11) {constB=-0.003}), Skew Amount=@n19, Is Vehicle=@n4, Uvs_TileRate=@n24) {materialFunction=&MF_Projection_Skew_Normal}
  n52 = MaterialFunctionCall(Caculated_Bounds=Add(A=@n17, B=Multiply(A=@n11) {constB=-0.004}), Skew Amount=@n19, Is Vehicle=@n4, Uvs_TileRate=@n24) {materialFunction=&MF_Projection_Skew_Normal}
  n53 = Add(A=Multiply(A=TextureSample(UVs=@n51.UVs, Tex=@n25).A) {constB=0.88}, B=Multiply(A=TextureSample(UVs=@n52.UVs, Tex=@n25).A) {constB=0.8})
  n54 = NamedRerouteDeclaration(None=@n6) {name=IsCharacter, nodeColor=<rgba 0.6641 1 0 1>}
  n55 = VectorParameter() {group=Wrap_FrozenLake, parameterName=FL_Layer_Masking_Tile_Rate, defaultValue=<rgba 1 2 0 4>}
  n56 = Multiply(A=@n55.RGB, B=@n55.RGB)
  n57 = VectorParameter() {group=Wrap_FrozenLake, parameterName=FL_Layer_Masking_Tile_Rate Vehicle, defaultValue=<rgba 1 2 0 4>}
  n58 = Multiply(A=@n57.RGB, B=@n57.RGB)
  n59 = VectorParameter() {group=Wrap_FrozenLake, parameterName=FL_Layer_Masking_Tile_Rate Mech, defaultValue=<rgba 1 2 0 4>}
  n60 = Multiply(A=@n59.RGB, B=@n59.RGB)
  n61 = VectorParameter() {group=Wrap_FrozenLake, parameterName=FL_Layer_Masking_Tile_Rate Character, defaultValue=<rgba 1 1 0 4>}
  n62 = Multiply(A=@n61.RGB, B=@n61.RGB)
  n63 = MaterialFunctionCall(
   Caculated_Bounds=@n16
   Skew Amount=LinearInterpolate(
    A=ScalarParameter() {defaultValue=0.5, group=Wrap_FrozenLake, parameterName=FL_Layer_Masking Skew}
    B=ScalarParameter() {defaultValue=0.5, group=Wrap_FrozenLake, parameterName=FL_Layer_Masking Skew Character}
    Alpha=@n54
   )
   Is Vehicle=@n4
   Uvs_TileRate=LinearInterpolate(A=LinearInterpolate(A=@n56, B=@n58, Alpha=@n4), B=LinearInterpolate(A=@n60, B=@n62, Alpha=@n54), Alpha=@n5)
  ) {materialFunction=&MF_Projection_Skew_Normal}
  n64 = Saturate(None=Add(
    A=TextureSample(UVs=@n63.UVs, Tex=@n46) {texture=&T_Wrap_Frozen_Mask_1}.R
    B=ScalarParameter() {defaultValue=0.1429, group=Wrap_FrozenLake, parameterName=FL_Layer_Masking_influance}
   ))
  n65 = Multiply(
   A=Saturate(None=Add(A=@n50, B=@n53))
   B=Add(A=@n64, B=ScalarParameter() {defaultValue=0.25, group=Wrap_FrozenLake, parameterName=FL_Layers_Masking}) {constB=0.25}
  )
  n66 = MaterialFunctionCall(Caculated_Bounds=Add(A=@n17, B=Multiply(A=@n11) {constB=-0.005}), Skew Amount=@n19, Is Vehicle=@n4, Uvs_TileRate=@n24) {materialFunction=&MF_Projection_Skew_Normal}
  n67 = MaterialFunctionCall(Caculated_Bounds=Add(A=@n17, B=Multiply(A=@n11) {constB=-0.006}), Skew Amount=@n19, Is Vehicle=@n4, Uvs_TileRate=@n24) {materialFunction=&MF_Projection_Skew_Normal}
  n68 = Add(A=Multiply(A=TextureSample(UVs=@n66.UVs, Tex=@n25).A) {constB=0.72}, B=Multiply(A=TextureSample(UVs=@n67.UVs, Tex=@n25).A) {constB=0.56})
  n69 = MaterialFunctionCall(Caculated_Bounds=Add(A=Multiply(A=@n11) {constB=-0.007}, B=@n17), Skew Amount=@n19, Is Vehicle=@n4, Uvs_TileRate=@n24) {materialFunction=&MF_Projection_Skew_Normal}
  n70 = MaterialFunctionCall(Caculated_Bounds=Add(A=@n17, B=Multiply(A=@n11) {constB=-0.008}), Skew Amount=@n19, Is Vehicle=@n4, Uvs_TileRate=@n24) {materialFunction=&MF_Projection_Skew_Normal}
  n71 = Add(A=Multiply(A=TextureSample(UVs=@n69.UVs, Tex=@n25).A) {constB=0.48}, B=Multiply(A=TextureSample(UVs=@n70.UVs, Tex=@n25).A) {constB=0.4})
  n72 = Multiply(A=@n11) {constB=-0.009}
  n73 = TextureSample(
   UVs=MaterialFunctionCall(Caculated_Bounds=Add(A=@n17, B=@n72), Skew Amount=@n19, Is Vehicle=@n4, Uvs_TileRate=@n24) {materialFunction=&MF_Projection_Skew_Normal}.UVs
   Tex=@n25
  )
  n74 = Multiply(A=@n11) {constB=-0.01}
  n75 = TextureSample(
   UVs=MaterialFunctionCall(Caculated_Bounds=Add(A=@n17, B=@n74), Skew Amount=@n19, Is Vehicle=@n4, Uvs_TileRate=@n24) {materialFunction=&MF_Projection_Skew_Normal}.UVs
   Tex=@n25
  )
  n76 = Add(A=Add(A=Multiply(A=@n73.A) {constB=0.32}, B=Multiply(A=@n75.A) {constB=0.16}), B=@n26) {constB=0}
  n77 = StaticSwitchParameter(True=@n68, False=Add(A=@n68, B=Add(A=@n71, B=@n76))) {group=Wrap_FrozenLake, parameterName="FL__Layer_Reduce_12--->6"}
  n78 = Add(A=@n77, B=@n36)
  n79 = Multiply(A=@n64, B=Saturate(None=StaticSwitchParameter(True=@n77, False=@n78) {group=Wrap_FrozenLake, parameterName=FL__Layer_Dirt_Remove}))
  n80 = Multiply(A=Saturate(None=Add(A=@n65, B=@n79)), B=ScalarParameter() {defaultValue=0.4952, group=Wrap_FrozenLake, parameterName=FL_Layers_Opacity})
  n81 = Saturate(None=LinearInterpolate(B=Add(A=@n48, B=@n80), Alpha=@n45.Masking Alternat Axies))
  n82 = CurveAtlasRowParameter(CurveTime=@n81) {
   atlas=/Game/Weapons/WeaponSkins/Wraps/Textures/CruveAtlas_Wraps_01
   curve=/Game/Weapons/WeaponSkins/Wraps/Textures/FrozenLake/C_Frozen_Color
   defaultValue=33
   group=Wrap_FrozenLake
   parameterName=FL_Color_Curve
  }
  n83 = NamedRerouteUsage()
  n84 = NamedRerouteUsage()
  n85 = MaterialFunctionCall(Color=@n7.Result) {materialFunction=&SplitComponents}
  n86 = LinearInterpolate(
   A=ScalarParameter() {defaultValue=-0.25, group=Wrap_FrozenLake, parameterName=FL_BackGround_Gradoffset}
   B=ScalarParameter() {defaultValue=-0.25, group=Wrap_FrozenLake, parameterName=FL_BackGround_Gradoffset Vehicle}
   Alpha=@n85.R
  )
  n87 = ScalarParameter() {defaultValue=-0.25, group=Wrap_FrozenLake, parameterName=FL_BackGround_Gradoffset Mech}
  n88 = LinearInterpolate(
   A=ScalarParameter() {defaultValue=0.5, group=Wrap_FrozenLake, parameterName=FL_BackGround_GradFalloff}
   B=ScalarParameter() {defaultValue=0.5, group=Wrap_FrozenLake, parameterName=FL_BackGround_GradFalloff Vehicle}
   Alpha=@n85.R
  )
  n89 = ScalarParameter() {defaultValue=0.5, group=Wrap_FrozenLake, parameterName=FL_BackGround_GradFalloff Mech}
  n90 = Power(Base=MaterialFunctionCall!(
    Bounds=LinearInterpolate(A=@n15, B=@n83, Alpha=@n84)
    ForwardAxis=VectorParameter() {group=Wrap_FrozenLake, parameterName=FL_BackGround_GradDirection, defaultValue=<rgba 0 0 1 1>}.RGB
    Offset=LinearInterpolate(A=@n86, B=@n87, Alpha=@n85.B)
    FallOff=LinearInterpolate(A=@n88, B=@n89, Alpha=@n85.B)
   ) {materialFunction=&/MF_Wrap_Gradent_Simple}.Raw_Data)
  n91 = Saturate(None=Multiply(A=@n90, B=ScalarParameter() {defaultValue=0.1, group=Wrap_FrozenLake, parameterName="FL_BackGround_GradOpacitcy "}))
  n92 = CurveAtlasRowParameter(CurveTime=@n91) {
   atlas=&CruveAtlas_Wraps_01
   curve=/Game/Weapons/WeaponSkins/Wraps/Textures/FrozenLake/C_Frozen_ColorBK
   defaultValue=34
   group=Wrap_FrozenLake
   parameterName=FL_Color_Curve_Background
  }
  n93 = MaterialFunctionCall!(
   Index=ScalarParameter() {defaultValue=34, group=Wrap_FrozenLake, parameterName="FL_Color_Curve_Background  Index"}
   CCinput=@n92
   CCinput_Alpha
   UV=@n91
  ) {materialFunction=&/MF_Global_CC_Tex}
  n94 = Abs(None=@n35)
  n95 = ComponentMask(None=@n94) {r=true}
  n96 = ComponentMask(None=@n94) {b=true}
  n97 = ComponentMask(None=@n94) {g=true}
  n98 = LinearInterpolate(
   A=Saturate(None=Add(A=@n95, B=@n96))
   B=Saturate(None=Add(A=@n96, B=@n97))
   Alpha=MaterialFunctionCall(Color=@n7.Result) {materialFunction=&SplitComponents}.R
  )
  n99 = ScalarParameter() {defaultValue=1, group=Wrap_FrozenLake, parameterName=FL_Tri_Planer_Mask_Contrast, sortPriority=0}
  n100 = Multiply(A=ScalarParameter() {defaultValue=0.5, group=Wrap_FrozenLake, parameterName=FL_BackGround_Thickness}, B=@n90)
  n101 = Multiply(
   A=Multiply(
    A=@n38.A
    B=MaterialFunctionCall(In=@n98, Contrast=@n99) {materialFunction=&CheapContrast}.Result
   )
   B=Saturate(None=OneMinus(None=@n100))
  )
  n102 = Add(
   A=MaterialFunctionCall(Index=ScalarParameter() {defaultValue=33, group=Wrap_FrozenLake, parameterName="FL_Color_Curve  Index"}, CCinput=@n82, UV=@n81) {materialFunction=&MF_Global_CC_Tex}.3V
   B=StaticSwitch(
    True=@n93.3V
    False=LinearInterpolate(A=@n93.3V, B=@n38.RGB, Alpha=@n101)
    Value=StaticBoolParameter() {group=Wrap_FrozenLake, parameterName=FL__Layer_Fish_Remove}
   )
  )
  n103 = Constant() {r=1}
  n104 = MaterialFunctionCall(Input=@n12, Input Low=Constant() {r=0.5}, Input High=@n103, Target Low=Constant(), Target High=@n103) {materialFunction=&RemapValueRange}
  n105 = ScalarParameter() {defaultValue=1, group=Wrap_FrozenLake, parameterName=FL_Edge_Multplyer, sortPriority=2}
  n106 = Multiply(
   A=Multiply(A=@n102, B=ScalarParameter() {defaultValue=1, group=Wrap_FrozenLake, parameterName=FL_Edge_Brightness, sortPriority=2})
   B=MaterialFunctionCall(In=Multiply(A=@n104.Result, B=@n105), Contrast=ScalarParameter() {group=Wrap_FrozenLake, parameterName=FL_Edge_Contrast, sortPriority=2}) {materialFunction=&CheapContrast}.Result
  )
  n107 = Constant() {r=0.5}
  n108 = Constant() {r=1}
  n109 = Constant() {r=0.5}
  n110 = Constant() {r=0.25}
  n111 = GetMaterialAttributes(None=@n39) {attributeGetTypes [#1]}
  n112 = Add(A=@n111.Normal, B=@n45.Raw_Normal)
  n113 = SetMaterialAttributes!(
   MaterialAttributes=@n40.MaterialAttributes
   BaseColor=@n41.RGB
   Metallic=ScalarParameter() {group=Wrap_FrozenLake, parameterName=FL_BS_Metalic, sortPriority=1}
   Specular=ScalarParameter() {defaultValue=0.5, group=Wrap_FrozenLake, parameterName=FL_BS_Specular, sortPriority=1}
   Roughness=LinearInterpolate(
    A=ScalarParameter() {defaultValue=0.2, group=Wrap_FrozenLake, parameterName=FL_BS_Roughness_Alternate_Axies, sortPriority=1}
    B=LinearInterpolate(A=@n42, B=@n43, Alpha=@n47.A)
    Alpha=@n45.Masking Alternat Axies
   )
   EmissiveColor=Multiply(
    A=Add(A=@n102, B=@n106)
    B=MaterialFunctionCall(Value_Morning=@n107, Value_Day=@n108, Value_Evening=@n109, Value_Night=@n110) {materialFunction=&Scalar_Time_of_Day}.Result
   )
   Normal=LinearInterpolate(A=@n111.Normal, B=@n112, Alpha=@n45.Masking Alternat Axies)
  ) {attributeSetTypes [
    #2
    #3
    #4
    #5
    #6
    #1
   ]}
  n114 = FunctionInput() {bUsePreviewValueAsDefault=true, inputName="Customization Mask (_CM B)", inputType=FunctionInput_Scalar, sortPriority=24}
  n115 = NamedRerouteUsage()
  n116 = NamedRerouteUsage()
  n117 = LinearInterpolate(
   A=@n114
   B=FunctionInput() {bUsePreviewValueAsDefault=true, inputName=Customization Mask B, inputType=FunctionInput_Scalar, sortPriority=24}
   Alpha=Saturate(None=Subtract(A=@n115, B=@n116))
  )
  FunctionOutput(None=StaticSwitchParameter(True=BlendMaterialAttributes(A=@n40.MaterialAttributes, B=@n113, Alpha=@n117) {vertexAttributeBlendType=UseA}, False=@n40.MaterialAttributes) {group=Customization_Wrap, parameterName=Wrap_FrozenLake})
  LinearInterpolate(
   A=@n111.Normal
   B=@n112
   Alpha=ScalarParameter() {defaultValue=1, group=Wrap_FrozenLake, parameterName=FL_BS_Normal_Intensty, sortPriority=1}
  )
  n118 = Max(A=@n2.Scale Y, B=@n2.Scale Y)
  Max(A=@n2.Scale X, B=@n118)
  Multiply(
   A=OneMinus(None=@n81)
   B=LinearInterpolate(
    A=VectorParameter() {group=Wrap_FrozenLake, parameterName=FL_BackGround_Color_A, defaultValue=<rgba 0 0 0 1>}.RGB
    B=VectorParameter() {group=Wrap_FrozenLake, parameterName=FL_BackGround_Color_B, defaultValue=<rgba 1 1 1 1>}.RGB
   )
  )
  Multiply(A=@n2.Scale X, B=@n118)
  NamedRerouteDeclaration(None=@n16) {name=Bounds, nodeColor=<rgba 1 0 0.5391 1>}
  NamedRerouteDeclaration(None=FunctionInput() {bUsePreviewValueAsDefault=true, inputName=isCharAccent, inputType=FunctionInput_Scalar, sortPriority=78}) {name=isCharAccent, nodeColor=<rgba 0 0.9253 1 1>}
  Saturate(None=Add(A=@n114, B=ScalarParameter() {group=Wrap_FrozenLake, parameterName=FL_Remove_Customization Mask, sortPriority=-10}))
 ]}
