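# Material function for the FruitPunch weapon wrap, a static-switched animated liquid with bobbing fruit and ice for weapons, vehicles and mechs.
# Warped, scrolling top/side projections of pre-skinned bounds supply ripple, bubble and normal detail over a height-blended two-color gradient.
# Two sine-animated fruit layers composite with outlines and underwater darkening, feeding time-of-day-tinted emissive with curvature edge boost.
# The customization mask blends the result in; the view-facing glow branch is unconnected.
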MaterialFunction /Game/Weapons/WeaponSkins/Wraps/MaterialFunctions/MF_Wrap_FruitPunch
graph {nodes [
  n1 = FunctionInput() {inputName=Material, inputType=FunctionInput_MaterialAttributes, sortPriority=-10}
  n2 = GetMaterialAttributes(None=@n1) {attributeGetTypes [#1, #2, #3]}
  n3 = TextureObjectParameter() {
   group=Wrap_FruitPunch
   parameterName=FP_Normal_01
   samplerType=SAMPLERTYPE_Normal
   sortPriority=4
   texture=/Game/Weapons/WeaponSkins/Wraps/Textures/FruitPunch/T_Wrap_FruitPunch_Bub_N
  }
  n4 = VertexInterpolator(VS=PreSkinnedPosition())
  n5 = PreSkinnedLocalBounds()
  n6 = FunctionInput(Preview=Divide(A=Subtract(A=@n4.PS, B=@n5.Min), B=Subtract(A=@n5.Max, B=@n5.Max))) {bUsePreviewValueAsDefault=true, inputName=Calcated_Bounds, sortPriority=11}
  n7 = VectorParameter() {group=Wrap_FruitPunch_UV, parameterName="FP_Scale_Weapon_01 A=uniform", sortPriority=30, defaultValue=<rgba 1 1 1 1>}
  n8 = VectorParameter() {group=Wrap_FruitPunch_UV, parameterName="FP_Scale_Vehicle_01 A=uniform", sortPriority=31, defaultValue=<rgba 1 1 1 1>}
  n9 = Multiply(A=@n8.RGB, B=@n8.RGB)
  n10 = VectorParameter() {group=Wrap_FruitPunch_UV, parameterName="FP_Scale_Vehicle_Side_01 A=uniform", sortPriority=31, defaultValue=<rgba 1 1 1 1>}
  n11 = Multiply(A=@n10.RGB, B=@n10.RGB)
  n12 = Constant()
  n13 = FunctionInput(Preview=@n12) {bUsePreviewValueAsDefault=true, inputName=Side Project Vehicle, inputType=FunctionInput_Scalar}
  n14 = FunctionInput(Preview=Constant()) {inputName=isVehicle, inputType=FunctionInput_Scalar, sortPriority=5}
  n15 = VectorParameter() {group=Wrap_FruitPunch_UV, parameterName="FP_Scale_Mech_01 A=uniform", defaultValue=<rgba 1 1 1 1>}
  n16 = FunctionInput(Preview=@n12) {bUsePreviewValueAsDefault=true, inputName=isMech, inputType=FunctionInput_Scalar, sortPriority=4}
  n17 = LinearInterpolate(
   A=LinearInterpolate(A=Multiply(A=@n7.RGB, B=@n7.RGB), B=LinearInterpolate(A=@n9, B=@n11, Alpha=@n13), Alpha=@n14)
   B=Multiply(A=@n15.RGB, B=@n15.RGB)
   Alpha=@n16
  )
  n18 = Time()
  n19 = VectorParameter() {group=Wrap_FruitPunch_UV, parameterName=FP_Speed_Weapon_01---------------------, sortPriority=10, defaultValue=<rgba 0 0 0 1>}
  n20 = VectorParameter() {group=Wrap_FruitPunch_UV, parameterName=FP_Speed_Vehicle_01, sortPriority=11, defaultValue=<rgba 0 0 0 1>}
  n21 = VectorParameter() {group=Wrap_FruitPunch_UV, parameterName=FP_Speed_Vehicle_Side_01, sortPriority=11, defaultValue=<rgba 0 0 0 1>}
  n22 = VectorParameter() {group=Wrap_FruitPunch_UV, parameterName=FP_Speed_Mech_01, sortPriority=12, defaultValue=<rgba 0 0 0 1>}
  n23 = LinearInterpolate(A=LinearInterpolate(A=@n19.RGB, B=LinearInterpolate(A=@n20.RGB, B=@n21.RGB, Alpha=@n13), Alpha=@n14), B=@n22.RGB, Alpha=@n16)
  n24 = Multiply(
   A=TextureCoordinate()
   B=LinearInterpolate(
    A=ScalarParameter() {defaultValue=1, group=Wrap_FruitPunch_UV, parameterName=FP-WarpUVS-01-TileRate Weapon, sortPriority=203}
    B=ScalarParameter() {defaultValue=1, group=Wrap_FruitPunch_UV, parameterName=FP-WarpUVS-01-TileRate Vehicle, sortPriority=204}
    Alpha=@n14
   )
  )
  n25 = VectorParameter() {group=Wrap_FruitPunch_UV, parameterName=FP-WarpUVS-01-Anitmated Weapon, sortPriority=207, defaultValue=<rgba 0 0 0 1>}
  n26 = VectorParameter() {group=Wrap_FruitPunch_UV, parameterName=FP-WarpUVS-01-Anitmated Vehicle, sortPriority=208, defaultValue=<rgba 0 0 0 1>}
  n27 = Multiply(A=Time() {period=1}, B=ComponentMask(None=LinearInterpolate(A=@n25.RGB, B=@n26.RGB, Alpha=@n14)) {g=true, r=true})
  n28 = TextureObjectParameter() {
   group=Wrap_FruitPunch
   parameterName=FP_Warp/Mask_01
   samplerType=SAMPLERTYPE_Masks
   sortPriority=3
   texture=/Game/Weapons/WeaponSkins/Wraps/Textures/FruitPunch/T_Wrap_FruitPunch_Bub_M
  }
  n29 = ChannelMaskParameter(None=TextureSample(UVs=Add(A=@n24, B=@n27), Tex=@n28).RGB) {group=Wrap_FruitPunch_UV, maskChannel=Green, parameterName=FP-WarpUVS-01_Channel, sortPriority=201, defaultValue=<rgba 0 0 0 1>}
  n30 = ScalarParameter() {group=Wrap_FruitPunch_UV, parameterName=FP-WarpUVS-01-Amount Weapon, sortPriority=205}
  n31 = ScalarParameter() {group=Wrap_FruitPunch_UV, parameterName=FP-WarpUVS-01-Amount_Vehicle, sortPriority=206}
  n32 = Add(A=Add(A=Multiply(A=@n6, B=@n17), B=Multiply(A=@n18, B=@n23)), B=Multiply(A=@n29.R, B=LinearInterpolate(A=@n30, B=@n31, Alpha=@n14)))
  n33 = ScalarParameter() {group=Wrap_FruitPunch, parameterName=FP_Blend_Contrast, sortPriority=10}
  n34 = ScalarParameter() {group=Wrap_FruitPunch, parameterName="FP_01Fill_alternate_Axis_Opacity (S)", sortPriority=10}
  n35 = FunctionInput(Preview=VertexInterpolator(VS=PreSkinnedNormal()).PS) {bUsePreviewValueAsDefault=true, inputName=Pre-Skinned Local Normal, sortPriority=10}
  n36 = LinearInterpolate(
   A=LinearInterpolate(
    A=Constant3Vector() {constant=<rgba 0 0 -0.25 1>}
    B=VectorParameter() {group=Wrap_FruitPunch_UV, parameterName=FP_Top_Side_Rotation_vehicle, defaultValue=<rgba 0 0 0 1>}.RGB
    Alpha=@n14
   )
   B=Constant3Vector() {constant=<rgba 0.25 0.25 0 1>}
   Alpha=@n16
  )
  n37 = MaterialFunctionCall!(
   Texture_Normals=@n3
   Rotation_Axis=@n36
   Caculated_Bounds=@n32
   Pre_Skinned_Normals=@n35
   Blend_Contrast=@n33
   Fill_alternate_Axis_Opacity=@n34
   Is_Vehicle=@n14
   Is_Mech=@n16
  ) {materialFunction=&/MF_TopsSideProjection_Normals}
  n38 = Constant() {r=1}
  n39 = MaterialFunctionCall(
   Input=FunctionInput() {bUsePreviewValueAsDefault=true, inputName=Customization Mask Curvature, inputType=FunctionInput_Scalar, sortPriority=24}
   Input Low=Constant() {r=0.5}
   Input High=@n38
   Target Low=Constant()
   Target High=@n38
  ) {materialFunction=&RemapValueRange}
  n40 = ScalarParameter() {defaultValue=1, group=Wrap_FruitPunch, parameterName=FP_Edge_Multplyer}
  n41 = LinearInterpolate(A=Constant() {r=1}, B=ScalarParameter() {defaultValue=1, group=Wrap_FruitPunch_UV, parameterName=FP_Fruit_Scale_Vehicle}, Alpha=@n14)
  n42 = ScalarParameter() {defaultValue=1, group=Wrap_FruitPunch_UV, parameterName=FP_Fruit_Scale_Backpack}
  n43 = NamedRerouteUsage()
  n44 = Constant3Vector() {constant=<rgba 1 1 -1 0>}
  n45 = Add(
   A=LinearInterpolate(
    A=Constant()
    B=VectorParameter() {group=Wrap_FruitPunch_UV, parameterName=FP_Fruit_offset_Vehicle, defaultValue=<rgba 0 0 0 1>}.RGB
    Alpha=@n14
   )
   B=Multiply(A=LinearInterpolate(A=@n41, B=@n42, Alpha=@n43), B=Multiply(A=@n6, B=@n44))
  )
  n46 = VectorParameter() {group=Wrap_FruitPunch, parameterName=FP_Fruit_Speed_02_Weapon, sortPriority=221, defaultValue=<rgba 0 0 0 1>}
  n47 = VectorParameter() {group=Wrap_FruitPunch, parameterName=FP_Fruit_Speed_02_Vehicle, sortPriority=221, defaultValue=<rgba 0 0 0 1>}
  n48 = VectorParameter() {group=Wrap_FruitPunch, parameterName=FP_Fruit_Speed_02_Vehicle_Side, sortPriority=221, defaultValue=<rgba 0 0 0 1>}
  n49 = LinearInterpolate(A=@n46.RGB, B=LinearInterpolate(A=@n47.RGB, B=@n48.RGB, Alpha=@n13), Alpha=@n14) {constB=0}
  n50 = VectorParameter() {group=Wrap_FruitPunch, parameterName=FP_Fruit_Speed_02_Mech, sortPriority=221, defaultValue=<rgba 0 0 0 1>}
  n51 = Sine(None=Multiply(A=Time(), B=LinearInterpolate(A=@n49, B=@n50.RGB, Alpha=@n16)))
  n52 = ScalarParameter() {defaultValue=0.01, group=Wrap_FruitPunch, parameterName=FP_Fruit_Speed_02_Sine_Strenght, sortPriority=220}
  n53 = MaterialFunctionCall(Color=Add(A=@n45, B=Multiply(A=@n51, B=@n52))) {materialFunction=&SplitComponents}
  n54 = AppendVector(A=@n53.G, B=@n53.G)
  n55 = ScalarParameter() {group=Wrap_FruitPunch_UV, parameterName=FP_Rotation_vehicle, sortPriority=200}
  n56 = MaterialFunctionCall(UVs=LinearInterpolate(A=AppendVector(A=@n53.R, B=@n53.R), B=AppendVector(A=@n53.R, B=@n53.R), Alpha=@n13), "Rotation Angle (0-1)"=@n55) {materialFunction=&CustomRotator}
  n57 = AppendVector(A=OneMinus(None=@n53.B), B=@n53.R)
  n58 = ScalarParameter() {group=Wrap_FruitPunch, parameterName=FP_Mech_Rotate}
  n59 = Multiply(A=@n29.R, B=ScalarParameter() {group=Wrap_FruitPunch, parameterName=FP_Fruit_UnderSurface_warp, sortPriority=199})
  n60 = Add(
   A=LinearInterpolate(
    A=LinearInterpolate(A=@n54, B=@n56.Rotated Values, Alpha=@n14)
    B=MaterialFunctionCall(UVs=@n57, "Rotation Angle (0-1)"=@n58) {desc=Mech Rotate, materialFunction=&CustomRotator}.Rotated Values
    Alpha=@n16
   )
   B=@n59
  )
  n61 = TextureObjectParameter() {group=Wrap_FruitPunch, parameterName=FP_Bobbing_Objects, texture=/Game/Weapons/WeaponSkins/Wraps/Textures/FruitPunch/T_Wrap_FruitPunch_BC}
  n62 = TextureSample(UVs=@n60, Tex=@n61) {texture=&T_Wrap_FruitPunch_BC}
  n63 = ScalarParameter() {defaultValue=1, group=Wrap_FruitPunch, parameterName=FP_Fruit_under_Sruface_Darken, sortPriority=199}
  n64 = TextureObjectParameter() {
   group=Wrap_FruitPunch
   parameterName=FP_Bobbing_Objects_mask
   texture=/Game/Weapons/WeaponSkins/Wraps/Textures/FruitPunch/T_Wrap_FruitPunch_M
  }
  n65 = TextureSample(UVs=@n60, Tex=@n64) {texture=&T_Wrap_FruitPunch_M}
  n66 = Abs(None=ComponentMask(None=@n35) {r=true})
  n67 = ScalarParameter() {defaultValue=1, group=Wrap_FruitPunch, parameterName=FP_Fruit_TriplanerMask_Contrast, sortPriority=199}
  n68 = MaterialFunctionCall(In=Abs(None=ComponentMask(None=@n35) {b=true}), Contrast=@n67) {materialFunction=&CheapContrast}
  n69 = MaterialFunctionCall(In=Abs(None=ComponentMask(None=@n35) {g=true}), Contrast=@n67) {materialFunction=&CheapContrast}
  n70 = LinearInterpolate(
   A=LinearInterpolate(
    A=MaterialFunctionCall(In=@n66, Contrast=@n67) {materialFunction=&CheapContrast}.Result
    B=LinearInterpolate(A=@n68.Result, B=@n69.Result, Alpha=@n13)
    Alpha=@n14
   )
   B=@n69.Result
   Alpha=@n16
  )
  n71 = Multiply(A=Multiply(A=@n65.G, B=@n65.G), B=@n70)
  n72 = ScalarParameter() {group=Wrap_FruitPunch, parameterName=FP_Fruit_under_Sruface_masking, sortPriority=199}
  n73 = Add(A=OneMinus(None=@n71), B=@n72)
  n74 = VectorParameter() {group=Wrap_FruitPunch, parameterName=FP_Fruit_Speed_01_Weapon, sortPriority=201, defaultValue=<rgba 0 0.1 0 1>}
  n75 = VectorParameter() {group=Wrap_FruitPunch, parameterName=FP_Fruit_Speed_01_Vehicle, sortPriority=201, defaultValue=<rgba 0 0 0 1>}
  n76 = VectorParameter() {group=Wrap_FruitPunch, parameterName=FP_Fruit_Speed_01_Vehicle_Side, sortPriority=201, defaultValue=<rgba 0 0 0 1>}
  n77 = LinearInterpolate(A=@n74.RGB, B=LinearInterpolate(A=@n75.RGB, B=@n76.RGB, Alpha=@n13), Alpha=@n14) {constB=0}
  n78 = VectorParameter() {group=Wrap_FruitPunch, parameterName=FP_Fruit_Speed_01_Mech, sortPriority=201, defaultValue=<rgba 0 0 0 1>}
  n79 = Sine(None=Multiply(A=Time(), B=LinearInterpolate(A=@n77, B=@n78.RGB, Alpha=@n16)))
  n80 = ScalarParameter() {defaultValue=0.01, group=Wrap_FruitPunch, parameterName=FP_Fruit_Speed_01Sine_Strenght, sortPriority=200}
  n81 = MaterialFunctionCall(Color=Add(A=@n45, B=Multiply(A=@n79, B=@n80))) {materialFunction=&SplitComponents}
  n82 = AppendVector(A=@n81.G, B=@n81.G) {desc=weapon}
  n83 = MaterialFunctionCall(UVs=LinearInterpolate(A=AppendVector(A=@n81.R, B=@n81.R), B=AppendVector(A=@n81.R, B=@n81.R), Alpha=@n13), "Rotation Angle (0-1)"=@n55) {materialFunction=&CustomRotator}
  n84 = AppendVector(A=OneMinus(None=@n81.B), B=@n81.R)
  n85 = Add(
   A=LinearInterpolate(
    A=LinearInterpolate(A=@n82, B=@n83.Rotated Values, Alpha=@n14)
    B=MaterialFunctionCall(UVs=@n84, "Rotation Angle (0-1)"=@n58) {desc=ismech, materialFunction=&CustomRotator}.Rotated Values
    Alpha=@n16
   ) {desc=Is mech}
   B=@n59
  )
  n86 = TextureSample(UVs=@n85, Tex=@n61) {texture=&T_Wrap_FruitPunch_BC}
  n87 = TextureSample(UVs=@n85, Tex=@n64) {texture=&T_Wrap_FruitPunch_M}
  n88 = Multiply(A=Multiply(A=@n87.R, B=@n87.R), B=@n70)
  n89 = OneMinus(None=@n88)
  n90 = VectorParameter() {group=Wrap_FruitPunch, parameterName=FP_Back_Color_B, sortPriority=50, defaultValue=<rgba 0 0.0592 1 1>}
  n91 = VectorParameter() {group=Wrap_FruitPunch, parameterName=FP_Back_Color_A, sortPriority=50, defaultValue=<rgba 0 0.0592 1 1>}
  n92 = VectorParameter() {group=Wrap_FruitPunch, parameterName="FP_01_Fill_alternate_Axis_Color_A=Roughness (V4)", sortPriority=10}
  n93 = MaterialFunctionCall!(
   IsCharacter
   IsStatsB
   Texture=@n28
   Rotation_Axis=@n36
   Caculated_Bounds=@n32
   Pre_Skinned_Normals=@n35
   Blend_Contrast=@n33
   Fill_alternate_Axis_Opacity=@n34
   "Fill_alternate_Axis_Color_A=Roughness"=AppendVector(A=@n92.RGB, B=@n92.RGB)
   Is_Vehicle=@n14
   Is_Mech=@n16
   Enable_Fill_Z_Weapon
   Fill_Z_Weapon_opacity
   Fill_Z_Weapon_Color
  ) {materialFunction=&/MF_TopsSideProjection_Base}
  n94 = ChannelMaskParameter(None=@n93.Combined) {group=Wrap_FruitPunch, maskChannel=Alpha, parameterName=FP-Mask-01_Channel, sortPriority=10, defaultValue=<rgba 0 0 0 1>}
  n95 = Add(
   A=Multiply(A=@n94.R, B=ScalarParameter() {defaultValue=1, group=Wrap_FruitPunch, parameterName=FP_Ripples_Opacity, sortPriority=9}) {constB=0.5}
   B=Multiply(
    A=ChannelMaskParameter(None=@n93.Combined) {group=Wrap_FruitPunch, parameterName=FP-Bubbles-01_Channel, sortPriority=10}.R
    B=ScalarParameter() {defaultValue=0.75, group=Wrap_FruitPunch, parameterName=FP_Bubbles_Opacity, sortPriority=9}
   ) {constB=0.5}
  )
  n96 = LinearInterpolate(A=@n90.RGB, B=@n91.RGB, Alpha=@n95)
  n97 = VectorParameter() {group=Wrap_FruitPunch, parameterName=FP_Front_Color_B, sortPriority=50, defaultValue=<rgba 1 0.8303 0 1>}
  n98 = VectorParameter() {group=Wrap_FruitPunch, parameterName=FP_Front_Color_A, sortPriority=50, defaultValue=<rgba 1 0.8303 0 1>}
  n99 = LinearInterpolate(A=@n97.RGB, B=@n98.RGB, Alpha=@n95)
  n100 = LinearInterpolate(
   A=LinearInterpolate(
    A=ScalarParameter() {group=Wrap_FruitPunch, parameterName=FP_Grad_Offset_Weapon, sortPriority=40}
    B=ScalarParameter() {group=Wrap_FruitPunch, parameterName=FP_Grad_Offset_Vehicle, sortPriority=40}
    Alpha=@n14
   )
   B=ScalarParameter() {group=Wrap_FruitPunch, parameterName=FP_Grad_Offset_Mech, sortPriority=40}
   Alpha=@n16
  )
  n101 = LinearInterpolate(
   A=LinearInterpolate(
    A=ScalarParameter() {defaultValue=0.5, group=Wrap_FruitPunch, parameterName=FP_Grad_Falloff_Weapon, sortPriority=40}
    B=ScalarParameter() {defaultValue=0.5, group=Wrap_FruitPunch, parameterName=FP_Grad_Falloff_Vehicle, sortPriority=40}
    Alpha=@n14
   )
   B=ScalarParameter() {defaultValue=0.5, group=Wrap_FruitPunch, parameterName=FP_Grad_Falloff_Mech, sortPriority=40}
   Alpha=@n16
  )
  n102 = VectorParameter() {group=Wrap_FruitPunch_UV, parameterName="FP_Foward_aixes_override A=1 off 0 on", sortPriority=200, defaultValue=<rgba 0 0 0 1>}
  n103 = Constant3Vector() {constant=<rgba 1 0 0 0>}
  n104 = Add(A=@n102.RGB, B=Multiply(A=@n102.A, B=FunctionInput(Preview=@n103) {bUsePreviewValueAsDefault=true, inputName=ForwardAxis}))
  n105 = MaterialFunctionCall(
   Transition Phase=Saturate(None=MaterialFunctionCall!(Bounds=@n6, ForwardAxis=@n104, Offset=@n100, FallOff=@n101) {materialFunction=&/MF_Wrap_Gradent_Simple}.Raw_Data)
   Height Texture=Multiply(A=@n94.R, B=ScalarParameter() {group=Wrap_FruitPunch, parameterName=FP_Grad_Height_Influce, sortPriority=39})
   Contrast=ScalarParameter() {group=Wrap_FruitPunch, parameterName=FP_Grad_Contrast, sortPriority=39}
  ) {materialFunction=&HeightLerp}
  n106 = VectorParameter() {group=Wrap_FruitPunch, parameterName=FP_Inner_Glow_Mult---------------, sortPriority=100, defaultValue=<rgba 1 1 1 1>}
  n107 = LinearInterpolate(
   A=Multiply(A=@n86.RGB, B=@n63)
   B=Multiply(A=LinearInterpolate(A=@n96, B=@n99, Alpha=@n105.Alpha), B=@n106.RGB)
   Alpha=Saturate(None=Add(A=@n72, B=@n89))
  )
  n108 = VectorParameter() {group=Wrap_FruitPunch, parameterName=FP_Fruit_Outline_Color, sortPriority=199, defaultValue=<rgba 0.7656 0 0.0447 1>}
  n109 = VectorParameter() {group=Wrap_FruitPunch, parameterName=FP_Ice_Outline_Color, sortPriority=199, defaultValue=<rgba 0 0 0 1>}
  n110 = ScalarParameter() {defaultValue=1, group=Wrap_FruitPunch, parameterName=FP_Bobbing_Offset_Fruit, sortPriority=199}
  n111 = ScalarParameter() {defaultValue=1, group=Wrap_FruitPunch, parameterName=FP_Bobbing_Offset_Ice, sortPriority=199}
  n112 = LinearInterpolate(A=Saturate(None=Power(Base=@n94.R, Exp=@n110)), B=Saturate(None=Power(Base=@n94.R, Exp=@n111)), Alpha=@n87.B)
  n113 = Multiply(A=@n88, B=@n112)
  n114 = Multiply(A=@n113, B=Add(A=@n113, B=@n112))
  n115 = ScalarParameter() {defaultValue=5, group=Wrap_FruitPunch, parameterName=FP_Ice_Outline_Contrast_A, sortPriority=199}
  n116 = MaterialFunctionCall(In=@n114, Contrast=@n115) {materialFunction=&CheapContrast}
  n117 = ScalarParameter() {defaultValue=1, group=Wrap_FruitPunch, parameterName=FP_Ice_Outline_Contrast_B, sortPriority=199}
  n118 = OneMinus(None=MaterialFunctionCall(In=@n114, Contrast=@n117) {materialFunction=&CheapContrast}.Result)
  n119 = LinearInterpolate(A=@n86.RGB, B=LinearInterpolate(A=@n108.RGB, B=@n109.RGB, Alpha=@n87.B), Alpha=Saturate(None=Multiply(A=@n116.Result, B=@n118)))
  n120 = ScalarParameter() {group=Wrap_FruitPunch, parameterName=FP_Fruit_Contrast, sortPriority=199}
  n121 = MaterialFunctionCall(In=@n114, Contrast=@n120) {materialFunction=&CheapContrast}
  n122 = LinearInterpolate(A=Saturate(None=Power(Base=@n94.R, Exp=@n110)), B=Saturate(None=Power(Base=@n94.R, Exp=@n111)), Alpha=@n65.B)
  n123 = Multiply(A=@n71, B=@n122)
  n124 = Multiply(A=@n123, B=Add(A=@n123, B=@n122))
  n125 = Multiply(
   A=MaterialFunctionCall(In=@n124, Contrast=@n115) {materialFunction=&CheapContrast}.Result
   B=OneMinus(None=MaterialFunctionCall(In=@n124, Contrast=@n117) {materialFunction=&CheapContrast}.Result)
  )
  n126 = LinearInterpolate(
   A=LinearInterpolate(A=Multiply(A=@n62.RGB, B=@n63), B=LinearInterpolate(A=@n107, B=@n119, Alpha=@n121.Result), Alpha=Saturate(None=@n73))
   B=LinearInterpolate(A=@n62.RGB, B=LinearInterpolate(A=@n108.RGB, B=@n109.RGB, Alpha=@n65.B), Alpha=Saturate(None=@n125))
   Alpha=MaterialFunctionCall(In=@n124, Contrast=@n120) {materialFunction=&CheapContrast}.Result
  )
  n127 = Multiply(
   A=MaterialFunctionCall(In=Multiply(A=@n39.Result, B=@n40), Contrast=ScalarParameter() {group=Wrap_FruitPunch, parameterName=FP_Edge_Contrast}) {materialFunction=&CheapContrast}.Result
   B=Multiply(A=ScalarParameter() {defaultValue=1, group=Wrap_FruitPunch, parameterName=FP_Edge_Brightness}, B=@n126)
  )
  n128 = VectorParameter() {group=Wrap_FruitPunch, parameterName=FP_TOD_Value_Morning, defaultValue=<rgba 0.5 0.5 0.5 1>}
  n129 = VectorParameter() {group=Wrap_FruitPunch, parameterName=FP_TOD_Value_Day, defaultValue=<rgba 1 1 1 1>}
  n130 = VectorParameter() {group=Wrap_FruitPunch, parameterName=FP_TOD_Value_Evening, defaultValue=<rgba 0.5 0.5 0.5 1>}
  n131 = VectorParameter() {group=Wrap_FruitPunch, parameterName=FP_TOD_Value_Night, defaultValue=<rgba 0.25 0.25 0.25 1>}
  n132 = SetMaterialAttributes!(
   MaterialAttributes=@n2.MaterialAttributes
   BaseColor=VectorParameter() {group=Wrap_FruitPunch, parameterName=FP_BaseColor, defaultValue=<rgba 0 0 0 1>}.RGB
   Roughness=ScalarParameter() {defaultValue=0.25, group=Wrap_FruitPunch, parameterName=FP_Roughness}
   Metallic=ScalarParameter() {group=Wrap_FruitPunch, parameterName=FP_Metalic}
   Normal=LinearInterpolate(
    A=@n2.Normal
    B=Add(A=@n2.Normal, B=@n37.Add_Surface_Normal)
    Alpha=ScalarParameter() {defaultValue=1, group=Wrap_FruitPunch, parameterName=FP_Normal_Intensty}
   )
   Specular=ScalarParameter() {defaultValue=0.5, group=Wrap_FruitPunch, parameterName=FP_Specular}
   EmissiveColor=Multiply(
    A=Add(A=@n127, B=@n126)
    B=MaterialFunctionCall(Color_Morning=@n128.RGB, Color_Day=@n129.RGB, Color_Evening=@n130.RGB, Color_Night=@n131.RGB) {materialFunction=&Color_Time_of_Day}.Result
   )
  ) {attributeSetTypes [
    #2
    #4
    #5
    #1
    #3
    #6
   ]}
  n133 = Saturate(None=Add(
    A=FunctionInput() {bUsePreviewValueAsDefault=true, inputName="Customization Mask (_CM B)", inputType=FunctionInput_Scalar, sortPriority=24}
    B=ScalarParameter() {group=Wrap_FruitPunch, parameterName=FB_Remove_Customization Mask, sortPriority=-10}
   ))
  FunctionOutput(None=StaticSwitchParameter(True=BlendMaterialAttributes(A=@n1, B=@n132, Alpha=@n133), False=@n1) {desc=" ", group=Customization_Wrap, parameterName=Wrap_FruitPunch}) {outputName=Result_1}
  n134 = VectorParameter() {group=Wrap_FruitPunch, parameterName=FP_Outer_Glow_Mult, sortPriority=100, defaultValue=<rgba 1 1 1 1>}
  n135 = ObjectPositionWS()
  n136 = Normalize(VectorInput=Subtract(A=@n135, B=CameraPositionWS()))
  n137 = Normalize(VectorInput=CrossProduct(A=@n136, B=Constant3Vector() {constant=<rgba 0 0 1 1>}))
  n138 = PreSkinnedLocalBounds()
  n139 = MaterialFunctionCall(Float3=@n138.Half Extents) {materialFunction=&BreakOutFloat3Components}
  n140 = Max(A=@n139.R, B=@n139.R)
  n141 = Divide(
   A=Multiply(
    A=MaterialFunctionCall() {materialFunction=/Engine/Functions/Engine_MaterialFunctions02/WorldPositionOffset/ObjectScale}.Scale X
    B=Max(A=@n140, B=@n139.B)
   )
   B=ScalarParameter() {defaultValue=1, group=Wrap_FruitPunch, parameterName=FP_UVRadialScaleZero, sortPriority=105}
  )
  n142 = Add(A=@n138.Min, B=FunctionInput() {inputName=Object Bounds Min Adjustment, sortPriority=51}) {constB=0}
  n143 = FunctionInput() {inputName=Object Bounds Max Adjustment, sortPriority=52}
  n144 = Multiply(A=Subtract(A=Add(A=@n138.Max, B=@n143) {constB=0}, B=@n142)) {constB=0.5}
  n145 = LinearInterpolate(A=TransformPosition(None=Add(A=@n142, B=@n144)) {transformType=TRANSFORMPOSSOURCE_World}, B=@n135, Alpha=@n16)
  LinearInterpolate(
   A=@n134.RGB
   Alpha=SphereMask(
    A=MaterialFunctionCall!(Plane Normal Axis=@n136, Plane X Axis=@n137, Plane Center=@n145, UV Scale=@n141) {materialFunction=/Game/Athena/Environments/Nexus/Materials/VirtualPlaneCoordinatesAbs}.UVs
    B=Constant() {r=0.5}
    Radius=ScalarParameter() {defaultValue=0.5, group=Wrap_FruitPunch, parameterName=FP_InnerGlow_Radius_A, sortPriority=105}
    Hardness=ScalarParameter() {defaultValue=-2, group=Wrap_FruitPunch, parameterName=FP_InnerGlow_Hardness_A, sortPriority=105}
   )
  )
  n146 = NamedRerouteDeclaration(None=FunctionInput() {bUsePreviewValueAsDefault=true, inputName=IsStatsB, inputType=FunctionInput_Vector2, sortPriority=78}) {name=Is_StatsB, nodeColor=<rgba 0 0.9253 1 1>}
  NamedRerouteDeclaration(None=ComponentMask(None=@n146) {r=true}) {name=IsBackpack, nodeColor=<rgba 0.2656 0 1 1>}
  NamedRerouteDeclaration(None=ComponentMask(None=@n146) {g=true}) {name=IsPickaxe, nodeColor=<rgba 1 0 0.5625 1>}
  StaticSwitchParameter(True=@n132, False=Constant3Vector()) {parameterName=dfasdf}
  VectorParameter() {group=Wrap_FruitPunch, parameterName=FP_Edge_Color, defaultValue=<rgba 0 0 0 1>}
 ]}
