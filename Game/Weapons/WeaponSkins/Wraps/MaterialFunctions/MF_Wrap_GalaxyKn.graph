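# Material function for the Galaxy Knight wrap skin, blended by primary and secondary masks behind a static switch.
# IsState inputs select projection UV settings per weapon, pickaxe, backpack, vehicle, mech or character.
# Projected masks and normals shape a view-angle curve-atlas metal with emissive layers: animated noise chest glow, timestamp-driven reactive fade, camera-vector triplanar stars and fresnel effects.
# An accent material blend exists but does not feed the output.
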MaterialFunction /Game/Weapons/WeaponSkins/Wraps/MaterialFunctions/MF_Wrap_GalaxyKn
graph {nodes [
  n1 = MaterialFunctionCall!(
   Camera _pos
   Axis
   Scale
   Scale_Multplyer
   Object-SafeZone-Virtual-Pivot=NamedRerouteUsage()
   Object-SafeZone-Virtual-CH-Switch=NamedRerouteUsage()
   SinglePlane Normal Axis
   SinglePlane X Axis
   PlaneCenter_distortion
  ) {materialFunction=&/MF_SafeZone_VirtualPlane}
  Add(
   A=ComponentMask(None=VectorParameter() {group=Wrap_GalaxyKnight_SubSurface, parameterName=GaKni_Chest_Center_Comp}.RGB) {g=true, r=true}
   B=@n1.Uvs
  )
  n2 = MaterialFunctionCall!(
   Normal_Map=NamedRerouteUsage()
   Offset Rt
   Pivot
   Caculated_Bounds=NamedRerouteUsage()
   Pre_Skinned_Normals=NamedRerouteUsage()
   is_State=NamedRerouteUsage()
   UV_Rotation=NamedRerouteUsage()
   Uvs_Offset
   Uvs_TileRate=NamedRerouteUsage()
   Uvs_Offset_Pivot=NamedRerouteUsage()
   For_Safe_Zones
   Rotation_Comp
   Clamp Y
   Clamp X
   Clamp Padding_Comp
   No_Rotation=StaticBoolParameter() {defaultValue=true, group=Wrap_GalaxyKnight_Uvs, parameterName=GaKni_No_Rotation}
  ) {materialFunction=&/MF_Projection_Side_Normal_V2}
  n3 = LinearInterpolate(
   A=NamedRerouteUsage()
   Alpha=MaterialFunctionCall(Float4=NamedRerouteUsage()) {materialFunction=&BreakOutFloat4Components}.B
  ) {constA=1}
  n4 = Multiply(A=@n2.Raw_Normal, B=@n3)
  n5 = OneMinus(None=NamedRerouteUsage())
  n6 = Add(
   A=NamedRerouteUsage()
   B=Multiply(A=Multiply(A=@n4, B=@n5), B=ScalarParameter() {defaultValue=0.4, group=Wrap_GalaxyKnight, parameterName=GaKni_Normal_Fresnel_Streangth})
  )
  n7 = Multiply(
   A=OneMinus(None=NamedRerouteUsage())
   B=MaterialFunctionCall!(
    Texture_Normals=NamedRerouteUsage()
    Rotation_Axis=NamedRerouteUsage()
    Caculated_Bounds=NamedRerouteUsage()
    Pre_Skinned_Normals=NamedRerouteUsage()
    IsStatsB=NamedRerouteUsage()
    IsState=NamedRerouteUsage()
    Blend_Contrast=ScalarParameter() {defaultValue=1, group=Wrap_GalaxyKnight_Detail_Normal, parameterName=GaKni_Alt_Contrast}
    Fill_alternate_Axis_Opacity=ScalarParameter() {defaultValue=1, group=Wrap_GalaxyKnight_Detail_Normal, parameterName=GaKni_Alt_Opacity}
    Scale=NamedRerouteUsage()
    Offset_Pivot
    Offset
   ) {materialFunction=&/MF_TopsSideProjection_Normals_V2}.Add_Surface_Normal
  )
  Add(
   A=@n6
   B=Multiply(A=@n7, B=ScalarParameter() {defaultValue=0.1, group=Wrap_GalaxyKnight_Detail_Normal, parameterName=GaKni_Normal_Detail_Streangth})
  )
  Add(A=NamedRerouteUsage()) {constB=-0.5}
  n8 = Time()
  n9 = ScalarParameter() {defaultValue=-0.05, group=Wrap_GalaxyKnight_Stars, parameterName=GaKni_Star_Rotation speed_B}
  n10 = NamedRerouteUsage()
  n11 = Time()
  n12 = VectorParameter() {group=Wrap_GalaxyKnight_Stars, parameterName=GaKni_Star_Pan_Speed_B, defaultValue=<rgba 0 0 0 1>}
  n13 = ComponentMask(None=MaterialFunctionCall!(
    CameraVector
    Rotation Speed=Multiply(A=@n8, B=@n9)
    Texture=@n10
    Tiling=ScalarParameter() {defaultValue=5, group=Wrap_GalaxyKnight_Stars, parameterName=GaKni_Star_Tiling_B}
    Offset=Multiply(A=@n11, B=@n12.RGB)
    AxisFadeContrast
   ) {materialFunction=/Game/Weapons/WeaponSkins/MaterialFunction/TriplanarCameraVector_Rotated}.Result) {r=true}
  n14 = MaterialFunctionCall!(
   CameraVector
   Texture=@n10
   Tiling=ScalarParameter() {defaultValue=3, group=Wrap_GalaxyKnight_Stars, parameterName=GaKni_Star_Tiling_A}
   Offset
   AxisFadeContrast
  ) {materialFunction=/Engine/Functions/Engine_MaterialFunctions01/Texturing/TriplanarCameraVector}
  n15 = ComponentMask(None=@n14.Result) {r=true}
  n16 = Multiply(A=TextureCoordinate()) {constB=0}
  n17 = Multiply(A=Time(), B=ScalarParameter() {defaultValue=0.05, group=Wrap_GalaxyKnight_Stars, parameterName=GaKni_Star_Flicker_Speed_A})
  n18 = TextureSample(UVs=Add(A=ComponentMask(None=@n14.Result) {g=true}, B=Add(A=@n16, B=@n17)), Tex=NamedRerouteUsage()) {samplerSource=SSM_Wrap_WorldGroupSettings, samplerType=SAMPLERTYPE_Masks}
  n19 = SmoothStep(Value=@n18.A) {constMax=0.9, constMin=0.6}
  n20 = Multiply(
   A=Multiply(A=Multiply(A=@n15, B=@n15), B=@n19) {constA=1}
   B=VectorParameter() {group=Wrap_GalaxyKnight_Stars, parameterName=GaKni_Star_Color_A, defaultValue=<rgba 15 15 15 1>}.RGB
  )
  Add(
   A=Multiply(
    A=Multiply(A=@n13, B=@n13)
    B=VectorParameter() {group=Wrap_GalaxyKnight_Stars, parameterName=GaKni_Star_Color_B, defaultValue=<rgba 0.5 0.5 0.5 1>}.RGB
   )
   B=@n20
  )
  n21 = NamedRerouteUsage()
  Add(A=@n21, B=NamedRerouteUsage())
  n22 = Multiply(
   A=Power(Base=NamedRerouteUsage()) {constExponent=3}
   B=ScalarParameter() {defaultValue=1, group=Wrap_GalaxyKnight_SubSurface, parameterName=GaKni_FresnelGlowStrength}
  )
  Add(A=@n22, B=NamedRerouteUsage())
  n23 = TextureSample() {samplerType=SAMPLERTYPE_Masks, texture=/Game/Weapons/FORT_Rifles/Textures/T_Scar_CM}
  n24 = FunctionInput(Preview=@n23.R) {bUsePreviewValueAsDefault=true, inputName=Mask Primary, inputType=FunctionInput_Scalar, sortPriority=24, previewValue=<1 0 0 1>}
  n25 = FunctionInput(Preview=@n23.B) {bUsePreviewValueAsDefault=true, inputName=Mask Secondary, inputType=FunctionInput_Scalar, sortPriority=24, previewValue=<1 0 0 1>}
  n26 = BlendMaterialAttributes(A=NamedRerouteUsage(), B=NamedRerouteUsage(), Alpha=Saturate(None=Add(A=@n24, B=@n25))) {vertexAttributeBlendType=UseA}
  BlendMaterialAttributes(A=@n26, B=NamedRerouteUsage(), Alpha=NamedRerouteDeclaration(None=@n25) {name=Msk_Accent, nodeColor=<rgba 0 1 0.7656 1>}) {vertexAttributeBlendType=UseA}
  Divide(A=@n18.A) {constB=0.5}
  FunctionOutput(None=StaticSwitchParameter(True=@n26, False=FunctionInput() {inputName=Material, inputType=FunctionInput_MaterialAttributes}) {group=Customization_Wrap, parameterName=Wrap_GalaxyKnight})
  MaterialFunctionCall(Float2=NamedRerouteUsage()) {materialFunction=&BreakOutFloat2Components}
  MaterialFunctionCall(Float2=NamedRerouteUsage()) {materialFunction=&BreakOutFloat2Components}
  n27 = Add(A=NamedRerouteUsage()) {constB=-0.5}
  n28 = VectorParameter() {group=Wrap_GalaxyKnight_Chest, parameterName=GaKni_Chest_Glow_Color_B, defaultValue=<rgba 1.9511 0.3461 2.5 1>}
  n29 = VectorParameter() {group=Wrap_GalaxyKnight_Chest, parameterName=GaKni_Chest_Glow_Color_A, defaultValue=<rgba 0.3459 0 0.5 1>}
  n30 = Multiply(A=NamedRerouteUsage()) {constB=-0.065}
  n31 = Multiply(A=NamedRerouteUsage(), B=ScalarParameter() {defaultValue=2, group=Wrap_GalaxyKnight_Chest, parameterName=GaKni_Chest_Noise_TileRate})
  n32 = VectorParameter() {group=Wrap_GalaxyKnight_Chest, parameterName=GaKni_Chest_Noise_Speed, defaultValue=<rgba -0.1 0.075 0 1>}
  n33 = Multiply(A=Time(), B=@n32.RGB)
  n34 = TextureSample(UVs=Add(A=Add(A=@n30, B=@n31), B=ComponentMask(None=@n33) {g=true, r=true}), Tex=NamedRerouteUsage()) {samplerType=SAMPLERTYPE_Masks}
  n35 = Power(Base=@n34.A) {constExponent=1.5}
  n36 = Subtract(A=ComponentMask(None=NamedRerouteUsage()) {g=true}) {constB=0.5}
  n37 = ScalarParameter() {defaultValue=10, group=Wrap_GalaxyKnight_Chest, parameterName=GaKni_Chest_Falloff}
  n38 = ScalarParameter() {defaultValue=10, group=Wrap_GalaxyKnight_Chest, parameterName=GaKni_Chest_Falloff_Vech}
  n39 = MaterialFunctionCall(Float4=NamedRerouteUsage()) {materialFunction=&BreakOutFloat4Components}
  n40 = ScalarParameter() {defaultValue=10, group=Wrap_GalaxyKnight_Chest, parameterName=GaKni_Chest_Falloff_Mech}
  n41 = ScalarParameter() {defaultValue=10, group=Wrap_GalaxyKnight_Chest, parameterName=GaKni_Chest_Falloff_Char}
  n42 = MaterialFunctionCall(Float2=NamedRerouteUsage()) {materialFunction=&BreakOutFloat2Components}
  n43 = LinearInterpolate(
   A=LinearInterpolate(A=LinearInterpolate(A=@n37, B=@n38, Alpha=@n39.R), B=LinearInterpolate(A=@n40, B=@n41, Alpha=@n39.A), Alpha=@n39.B)
   B=ScalarParameter() {defaultValue=4, group=Wrap_GalaxyKnight_Chest, parameterName=GaKni_Chest_Falloff_Back}
   Alpha=@n42.R
  )
  n44 = ScalarParameter() {defaultValue=4, group=Wrap_GalaxyKnight_Chest, parameterName=GaKni_Chest_Falloff_Pickaxe}
  n45 = Saturate(None=Multiply(A=Abs(None=@n36), B=LinearInterpolate(A=@n43, B=@n44, Alpha=@n42.G)) {constB=5})
  n46 = SmoothStep(Value=Add(A=Multiply(A=@n35), B=Multiply(A=@n45) {constB=0.75}))
  n47 = Add(A=Saturate(None=Multiply(A=@n27) {constB=0.5}), B=LinearInterpolate(A=@n28.RGB, B=@n29.RGB, Alpha=Multiply(A=@n46)))
  Multiply(A=@n47)
  Multiply(A=@n19)
  n48 = Multiply(
   A=ScalarParameter() {defaultValue=1, group=Wrap_GalaxyKnight_Chest, parameterName=GaKni_Chest_Glow_Brightness}
   B=Multiply(A=NamedRerouteUsage(), B=@n47)
  )
  Multiply(A=@n48)
  n49 = NamedRerouteUsage()
  n50 = LinearInterpolate(
   A=ScalarParameter() {defaultValue=1, group=Wrap_GalaxyKnight, parameterName=GaKni_Metal_Outer}
   B=ScalarParameter() {group=Wrap_GalaxyKnight, parameterName=GaKni_Metal_Inner}
   Alpha=@n49
  )
  Multiply(A=@n50)
  Multiply(B=@n34.A) {constA=1}
  n51 = NamedRerouteDeclaration(None=@n2.UVs) {name=UVs, nodeColor=<rgba 1 0 0.0703 1>}
  n52 = TextureSample(UVs=@n51, Tex=NamedRerouteUsage()) {samplerType=SAMPLERTYPE_Normal}
  Multiply(A=@n52.R)
  Multiply(A=MaterialFunctionCall(Value_Morning=Constant() {r=0.5}, Value_Day=Constant() {r=1}, Value_Evening=Constant() {r=0.5}, Value_Night=Constant() {r=0.25}) {materialFunction=&Scalar_Time_of_Day}.Result)
  Multiply()
  n53 = NamedRerouteUsage()
  Multiply(A=@n53, B=OneMinus(None=NamedRerouteUsage()))
  n54 = ScalarParameter() {defaultValue=2, group=Wrap_GalaxyKnight_Uvs, parameterName=GaKni_Scale_Detail_Vehicle, sortPriority=150}
  n55 = ScalarParameter() {defaultValue=2, group=Wrap_GalaxyKnight_Uvs, parameterName=GaKni_Scale_Detail_Vehicle_Side, sortPriority=150}
  n56 = MaterialFunctionCall(Float4=NamedRerouteUsage()) {materialFunction=&BreakOutFloat4Components}
  n57 = LinearInterpolate(
   A=LinearInterpolate(
    A=ScalarParameter() {defaultValue=2, group=Wrap_GalaxyKnight_Uvs, parameterName=GaKni_Scale_Detail_Weapon, sortPriority=150}
    B=LinearInterpolate(A=@n54, B=@n55, Alpha=@n56.G)
    Alpha=@n56.R
   )
   B=LinearInterpolate(
    A=ScalarParameter() {defaultValue=2, group=Wrap_GalaxyKnight_Uvs, parameterName=GaKni_Scale_Detail_Mech, sortPriority=150}
    B=ScalarParameter() {defaultValue=2, group=Wrap_GalaxyKnight_Uvs, parameterName=GaKni_Scale_Detail_Character, sortPriority=150}
    Alpha=@n56.A
   )
   Alpha=@n56.B
  )
  n58 = ScalarParameter() {defaultValue=2, group=Wrap_GalaxyKnight_Uvs, parameterName=GaKni_Scale_Detail_Backpack, sortPriority=150}
  n59 = MaterialFunctionCall(Float2=NamedRerouteUsage()) {materialFunction=&BreakOutFloat2Components}
  NamedRerouteDeclaration(None=LinearInterpolate(
    A=LinearInterpolate(A=@n57, B=@n58, Alpha=@n59.R)
    B=ScalarParameter() {defaultValue=2, group=Wrap_GalaxyKnight_Uvs, parameterName=GaKni_Scale_Detail_Pickaxe, sortPriority=150}
    Alpha=@n59.G
   )) {name=UVs_Scale_Detail, nodeColor=<rgba 0 0.7188 1 1>}
  n60 = PixelNormalWS()
  n61 = Fresnel(Normal=@n60) {baseReflectFraction=0, exponent=1}
  n62 = Round(None=ScalarParameter() {defaultValue=1, group=Wrap_GalaxyKnight_Chest, parameterName=GaKni_Chest_Fres_Inverted, sliderMax=1})
  n63 = Saturate(None=Abs(None=Subtract(A=@n61, B=@n62)))
  n64 = ScalarParameter() {defaultValue=1, group=Wrap_GalaxyKnight_Chest, parameterName=GaKni_Chest_Fres_Power, sliderMax=10}
  NamedRerouteDeclaration(None=Clamp(None=Power(Base=@n63, Exp=@n64)) {maxDefault=100000}) {name=Chest_Fres, nodeColor=<rgba 1 0.7266 0 1>}
  n65 = TextureSample() {samplerType=SAMPLERTYPE_Normal, texture=/Game/Weapons/FORT_Rifles/T_SCAR_N}
  n66 = Constant3Vector() {constant=<rgba 1 0 0 0>}
  n67 = GetMaterialAttributes(None=FunctionInput(Preview=SetMaterialAttributes!(MaterialAttributes, Normal=@n65.RGB, BaseColor=@n66) {attributeSetTypes [#1, #2]}) {inputName=Material_Root, inputType=FunctionInput_MaterialAttributes}) {attributeGetTypes [
    #2
    #3
    #1
    #4
   ]}
  NamedRerouteDeclaration(None=@n67.MaterialAttributes) {name=Root_Material, nodeColor=<rgba 0.7812 1 0 1>}
  NamedRerouteDeclaration(None=FunctionInput(Preview=Constant2Vector()) {bUsePreviewValueAsDefault=true, inputName=IsState_B, inputType=FunctionInput_Vector2, sortPriority=4}) {name=IsState_B, nodeColor=<rgba 0.1497 1 0.1633 1>}
  n68 = PreSkinnedLocalBounds()
  n69 = Subtract(A=VertexInterpolator(VS=PreSkinnedPosition()).PS, B=@n68.Min)
  n70 = MaterialFunctionCall(Color=Subtract(A=@n68.Max, B=@n68.Max)) {materialFunction=&SplitComponents}
  n71 = Max(A=Max(A=@n70.R, B=@n70.R), B=@n70.B)
  NamedRerouteDeclaration(None=FunctionInput(Preview=Divide(A=@n69, B=@n71)) {bUsePreviewValueAsDefault=true, inputName=Calcated_Bounds, sortPriority=11}) {name=Caclcated_Bounds, nodeColor=<rgba 0 0.9297 1 1>}
  n72 = PreSkinnedNormal()
  NamedRerouteDeclaration(None=FunctionInput(Preview=VertexInterpolator(VS=@n72).PS) {bUsePreviewValueAsDefault=true, inputName=Pre-Skinned Local Normal, sortPriority=10}) {name=PreSkin_Normal, nodeColor=<rgba 0.8281 1 0 1>}
  n73 = FunctionInput(Preview=Constant4Vector() {constant=<rgba 0 0 0 -1.5>}.A) {bUsePreviewValueAsDefault=true, inputName=Object-SafeZone-Virtual-Pivot, inputType=FunctionInput_Vector4, sortPriority=105}
  NamedRerouteDeclaration(None=@n73) {name=SafeZone VP, nodeColor=<rgba 0.5 1 0 1>}
  n74 = FunctionInput(Preview=Constant()) {bUsePreviewValueAsDefault=true, inputName=Object-SafeZone-VP-Character-PreView-Switch, inputType=FunctionInput_Scalar, sortPriority=105}
  NamedRerouteDeclaration(None=@n74) {name=SafeZone Character-Switch, nodeColor=<rgba 1 0.9375 0 1>}
  n75 = VectorParameter() {group=Wrap_GalaxyKnight_Metal_LUT, parameterName=GaKni_BaseColor, defaultValue=<rgba 0.3325 0.3467 0.5776 1>}
  n76 = Clamp(None=DotProduct(A=CameraVectorWS(), B=PixelNormalWS()))
  n77 = CurveAtlasRowParameter(CurveTime=@n76) {
   atlas=/Game/Weapons/WeaponSkins/Wraps/Textures/CruveAtlas_Wraps_01
   curve=/Game/Weapons/WeaponSkins/Wraps/Textures/0_ColorCurves/CC_GalKnight
   defaultValue=122
   group=Wrap_GalaxyKnight_Metal_LUT
   parameterName=GaKni_Metal_LUT_CC
   primitiveDataIndex=1
   sortPriority=-5
  }
  n78 = ChannelMaskParameter(None=MaterialFunctionCall!(
    Index=ScalarParameter() {defaultValue=122, group=Wrap_GalaxyKnight_Metal_LUT, parameterName="GaKni_Metal_LUT_CC  Index", sortPriority=-5}
    CCinput=@n77
    CCinput_Alpha=@n77
    UV=@n76
   ) {materialFunction=&/MF_Global_CC_Tex}.4v) {group=Wrap_GalaxyKnight_Metal_LUT, parameterName=GaKni_Metal_LUT_Channel, sortPriority=-5}
  n79 = TextureSample(UVs=@n51, Tex=NamedRerouteUsage()) {samplerType=SAMPLERTYPE_Normal}
  n80 = MaterialFunctionCall(Float4=Multiply(A=Multiply(A=@n79.RGBA, B=@n3), B=@n5)) {materialFunction=&BreakOutFloat4Components}
  n81 = OneMinus(None=@n80.R)
  n82 = ScalarParameter() {defaultValue=0.15, group=Wrap_GalaxyKnight, parameterName=GaKni_Roughness_outer}
  n83 = ScalarParameter() {defaultValue=1, group=Wrap_GalaxyKnight, parameterName=GaKni_Roughness_Inner}
  n84 = Multiply(A=@n47, B=NamedRerouteUsage())
  n85 = Divide(
   A=Subtract(A=Time(), B=ScalarParameter() {defaultValue=-9999999, group=Wrap_Reactive, parameterName=TimeFrom-BP----------------, sortPriority=600})
   B=ScalarParameter() {defaultValue=0.25, group=Wrap_GalaxyKnight_Reactive, parameterName=GaKni_Reactive_Duration, sortPriority=5}
  )
  n86 = LinearInterpolate(
   A=ScalarParameter() {defaultValue=1, group=Wrap_GalaxyKnight_Reactive, parameterName=GaKni_Reactive_Start, sortPriority=10}
   B=ScalarParameter() {group=Wrap_GalaxyKnight_Reactive, parameterName=GaKni_Reactive_End, sortPriority=10}
   Alpha=Power(
    Base=Saturate(None=@n85)
    Exp=ScalarParameter() {defaultValue=0.25, group=Wrap_GalaxyKnight_Reactive, parameterName=GaKni_Reactive_Falloff, sortPriority=5}
   )
  )
  n87 = Add(A=@n53, B=Add(A=@n48, B=Multiply(A=@n84, B=@n86)))
  n88 = Multiply(A=@n21, B=OneMinus(None=NamedRerouteUsage()))
  n89 = SetMaterialAttributes!(
   MaterialAttributes=NamedRerouteUsage()
   BaseColor=Multiply(A=Multiply(A=@n75.RGB, B=@n78.R), B=@n81)
   Roughness=LinearInterpolate(
    A=LinearInterpolate(A=@n82, B=@n83, Alpha=@n49)
    B=ScalarParameter() {defaultValue=0.15, group=Wrap_GalaxyKnight, parameterName=GaKni_Roughness_Glow}
    Alpha=@n80.R
   )
   Specular=ScalarParameter() {defaultValue=0.5, group=Wrap_GalaxyKnight, parameterName=GaKni_Spec}
   Metallic=Multiply(A=@n81, B=@n50)
   Normal=NamedRerouteUsage()
   EmissiveColor=Clamp(None=Add(A=@n87, B=@n88)) {maxDefault=100}
  ) {attributeSetTypes [
    #2
    #3
    #5
    #4
    #1
    #6
   ]}
  NamedRerouteDeclaration(None=@n89) {name=Mat_Base, nodeColor=<rgba 0 0.3906 1 1>}
  NamedRerouteDeclaration(None=TextureObjectParameter() {group=Wrap_GalaxyKnight_Stars, parameterName=GaKni__Tex_Msk, texture=&T_BlacksRGB}) {name=Tex_Msk, nodeColor=<rgba 1 0.75 0 1>}
  n90 = FunctionInput(Preview=@n23.G) {bUsePreviewValueAsDefault=true, inputName=Mask Crevise, inputType=FunctionInput_Scalar, sortPriority=24, previewValue=<1 0 0 1>}
  NamedRerouteDeclaration(None=@n90) {name=Msk_Curvature, nodeColor=<rgba 0 1 0.7422 1>}
  n91 = TextureObjectParameter() {
   group=Wrap_GalaxyKnight
   parameterName=GaKni__Tex_Nml
   samplerType=SAMPLERTYPE_Normal
   texture=&FlatNormalBlue
  }
  NamedRerouteDeclaration(None=@n91) {name=Tex_Nml, nodeColor=<rgba 1 0.75 0 1>}
  n92 = Add(A=NamedRerouteUsage()) {constB=-0.5}
  n93 = Add(A=NamedRerouteUsage()) {constB=-0.5}
  n94 = Add(
   A=@n1.Uvs
   B=Multiply(A=Add(A=@n92, B=@n93), B=ScalarParameter() {defaultValue=-0.25, group=Wrap_GalaxyKnight_SubSurface, parameterName=GaKni_ChestCurve_Warp}) {constB=-0.2}
  )
  n95 = ScalarParameter() {defaultValue=0.5, group=Wrap_GalaxyKnight_SubSurface, parameterName=GaKni_ChestGlowRadius}
  n96 = NamedRerouteUsage()
  n97 = Power(
   Base=SphereMask(A=@n94, B=Constant() {r=0.5}, Radius=Multiply(A=@n95, B=@n96)) {hardnessPercent=0}
   Exp=ScalarParameter() {defaultValue=3, group=Wrap_GalaxyKnight_SubSurface, parameterName=GaKni_ChestGlowPower}
  ) {constExponent=5}
  n98 = ScalarParameter() {defaultValue=0.1, group=Wrap_GalaxyKnight_SubSurface, parameterName=GaKni_ChestGlowRadius_Inner}
  n99 = Power(Base=SphereMask(A=@n94, B=Constant() {r=0.5}, Radius=Multiply(A=@n98, B=@n96)) {hardnessPercent=0})
  NamedRerouteDeclaration(None=Multiply(A=Add(A=@n97, B=@n99), B=ScalarParameter() {defaultValue=0.5, group=Wrap_GalaxyKnight_SubSurface, parameterName=GaKni_ChestGlowStrength})) {name=ChestGlow_1, nodeColor=<rgba 0 1 0.3906 1>}
  n100 = LinearInterpolate(Alpha=Fresnel(ExponentIn=ScalarParameter() {defaultValue=0.5, group=Wrap_GalaxyKnight, parameterName=GaKni_InnerGlowFresnelExp}, Normal=@n60) {baseReflectFraction=0.05, exponent=1}) {constA=1, constB=-0.5}
  NamedRerouteDeclaration(None=Power(Base=Saturate(None=@n100)) {constExponent=3}) {name=InnerFresnelMask, nodeColor=<rgba 0 0.0625 1 1>}
  n101 = VectorParameter() {group=Wrap_GalaxyKnight_SubSurface, parameterName=GaKni_SubsurfaceColor, defaultValue=<rgba 0.3228 0.1047 0.3854 0.5>}
  NamedRerouteDeclaration(None=Multiply(A=Multiply(A=@n22, B=@n101.RGB), B=@n101.A)) {name=SubsurfaceColor_1, nodeColor=<rgba 0 0.7422 1 1>}
  n102 = MaterialFunctionCall(Color=Abs(None=NamedRerouteUsage())) {materialFunction=&SplitComponents}
  n103 = MaterialFunctionCall(Float4=NamedRerouteUsage()) {materialFunction=&BreakOutFloat4Components}
  n104 = LinearInterpolate(A=@n102.R, B=LinearInterpolate(A=@n102.B, B=@n102.B, Alpha=@n103.G), Alpha=@n103.R)
  NamedRerouteDeclaration(None=MaterialFunctionCall(
    In=LinearInterpolate(A=@n104, B=@n102.G, Alpha=@n103.B)
    Contrast=ScalarParameter() {defaultValue=1, group=Wrap_GalaxyKnight, parameterName=GaKni_Tri_Blend_Contrast}
   ) {materialFunction=&CheapContrast}.Result) {name=Tri_Mask, nodeColor=<rgba 1 0.4688 0 1>}
  NamedRerouteDeclaration(None=@n6) {name=FresnelNormals, nodeColor=<rgba 0.5469 1 0 1>}
  NamedRerouteDeclaration(None=@n20) {name=Stars, nodeColor=<rgba 0 1 0.6484 1>}
  NamedRerouteDeclaration(None=@n6) {name=Normals, nodeColor=<rgba 0.5469 1 0 1>}
  NamedRerouteDeclaration(None=TextureObjectParameter() {group=Wrap_GalaxyKnight, parameterName=GaKni__Tex_Msk1, texture=&T_BlacksRGB}) {name=Tex_Msk1, nodeColor=<rgba 1 0.75 0 1>}
  n105 = TextureObjectParameter() {
   group=Wrap_GalaxyKnight_Detail_Normal
   parameterName=GaKni__Tex_Nml1
   samplerType=SAMPLERTYPE_Normal
   texture=&FlatNormalBlue
  }
  NamedRerouteDeclaration(None=@n105) {name=Tex_Nml1, nodeColor=<rgba 1 0.75 0 1>}
  NamedRerouteDeclaration(None=LinearInterpolate(B=@n52.A, Alpha=@n3) {constA=0.5, constB=0.5}) {name=Msk_Detal_Curvature, nodeColor=<rgba 0 1 0.3672 1>}
  n106 = VectorParameter() {group=Wrap_GalaxyKnight_Uvs, parameterName="GaKni_Detail_Rotation_Weapon\\Pickaxe", defaultValue=<rgba 0 0 -0.25 1>}
  n107 = VectorParameter() {group=Wrap_GalaxyKnight_Uvs, parameterName=GaKni_Detail_Rotation_Vehicle_Side, defaultValue=<rgba 0 0 0 1>}
  n108 = MaterialFunctionCall(Float4=NamedRerouteUsage()) {materialFunction=&BreakOutFloat4Components}
  n109 = LinearInterpolate(
   A=VectorParameter() {group=Wrap_GalaxyKnight_Uvs, parameterName=GaKni_Detail_Rotation_Vehicle, defaultValue=<rgba 0 0 0 1>}.RGB
   B=@n107.RGB
   Alpha=@n108.G
  )
  n110 = VectorParameter() {group=Wrap_GalaxyKnight_Uvs, parameterName="GaKni_Detail_Rotation_Mech\\Character", defaultValue=<rgba 0.25 0.25 0 1>}
  NamedRerouteDeclaration(None=LinearInterpolate(A=LinearInterpolate(A=@n106.RGB, B=@n109, Alpha=@n108.R), B=@n110.RGB, Alpha=@n108.B)) {name=TopsIde_Rotation, nodeColor=<rgba 0.6406 0 1 1>}
  NamedRerouteDeclaration(None=TextureObjectParameter() {group=Wrap_GalaxyKnight, parameterName=GaKni__Tex_Refl, texture=&T_BlacksRGB}) {name=Tex_Refl, nodeColor=<rgba 1 0.75 0 1>}
  n111 = Transform(None=Multiply(A=@n7, B=ScalarParameter() {defaultValue=1, group=Wrap_GalaxyKnight_Detail_Normal, parameterName=GaKni_Normal_Detail_Streangth_Refl})) {transformSourceType=TRANSFORMSOURCE_World, transformType=TRANSFORM_Tangent}
  n112 = TextureSample(UVs=ComponentMask(None=ReflectionVectorWS(CustomWorldNormal=@n111)) {g=true, r=true}, Tex=NamedRerouteUsage())
  n113 = ScalarParameter() {defaultValue=0.15, group=Wrap_GalaxyKnight_Detail_Normal, parameterName=GaKni_ReflectionStrength}
  NamedRerouteDeclaration(None=Multiply(A=NamedRerouteUsage(), B=Multiply(A=@n112.RGB, B=@n113))) {name=Gal_reflection, nodeColor=<rgba 0 1 0.4609 1>}
  NamedRerouteDeclaration(None=Power(Base=Saturate(None=@n100))) {name=InnerFresnelMask_1, nodeColor=<rgba 0 0.0625 1 1>}
  NamedRerouteDeclaration(None=@n71) {name=scale, nodeColor=<rgba 0.6172 0 1 1>}
  n114 = Multiply(A=@n52.RGBA, B=@n3)
  NamedRerouteDeclaration(None=MaterialFunctionCall(Float4=Multiply(A=@n114, B=@n5)) {materialFunction=&BreakOutFloat4Components}.R) {name=Chest_Msk, nodeColor=<rgba 0 1 0.4375 1>}
  NamedRerouteDeclaration(None=@n80.A) {name=Chest_Msk_MskB, nodeColor=<rgba 0 1 0.4375 1>}
  n115 = VectorParameter() {group=Wrap_GalaxyKnight_Uvs, parameterName=GaKni_Offset_Vehicle, defaultValue=<rgba 0 0 0 1>}
  n116 = VectorParameter() {group=Wrap_GalaxyKnight_Uvs, parameterName=GaKni_Offset_VehicleSide, defaultValue=<rgba 0 0 0 1>}
  n117 = MaterialFunctionCall(Float4=NamedRerouteUsage()) {materialFunction=&BreakOutFloat4Components}
  n118 = LinearInterpolate(
   A=LinearInterpolate(
    A=VectorParameter() {group=Wrap_GalaxyKnight_Uvs, parameterName=GaKni_Offset_Weapon, defaultValue=<rgba 0 0 0 1>}.RGB
    B=LinearInterpolate(A=@n115.RGB, B=@n116.RGB, Alpha=@n117.G)
    Alpha=@n117.R
   )
   B=LinearInterpolate(
    A=VectorParameter() {group=Wrap_GalaxyKnight_Uvs, parameterName=GaKni_Offset_Mech, defaultValue=<rgba 0 0 0 1>}.RGB
    B=VectorParameter() {group=Wrap_GalaxyKnight_Uvs, parameterName=GaKni_Offset_Character, defaultValue=<rgba 0 0 0 1>}.RGB
    Alpha=@n117.A
   )
   Alpha=@n117.B
  )
  n119 = VectorParameter() {group=Wrap_GalaxyKnight_Uvs, parameterName=GaKni_Offset_Backpack, defaultValue=<rgba 0 0 0 1>}
  n120 = MaterialFunctionCall(Float2=NamedRerouteUsage()) {materialFunction=&BreakOutFloat2Components}
  NamedRerouteDeclaration(None=LinearInterpolate(
    A=LinearInterpolate(A=@n118, B=@n119.RGB, Alpha=@n120.R)
    B=VectorParameter() {group=Wrap_GalaxyKnight_Uvs, parameterName=GaKni_Offset_Pickaxe, defaultValue=<rgba 0 0 0 1>}.RGB
    Alpha=@n120.G
   )) {name=UVs_Offset, nodeColor=<rgba 0 0.625 1 1>}
  n121 = ScalarParameter() {group=Wrap_GalaxyKnight_Uvs, parameterName="GaKni_Rotation_Weapon\\Pickaxe", sortPriority=150}
  n122 = MaterialFunctionCall(Float4=NamedRerouteUsage()) {materialFunction=&BreakOutFloat4Components}
  n123 = LinearInterpolate(
   A=ScalarParameter() {group=Wrap_GalaxyKnight_Uvs, parameterName=GaKni_Rotation_Vehicle, sortPriority=150}
   B=ScalarParameter() {group=Wrap_GalaxyKnight_Uvs, parameterName=GaKni_Rotation_Vehicle_Side, sortPriority=150}
   Alpha=@n122.G
  )
  NamedRerouteDeclaration(None=LinearInterpolate(
    A=LinearInterpolate(A=@n121, B=@n123, Alpha=@n122.R)
    B=ScalarParameter() {group=Wrap_GalaxyKnight_Uvs, parameterName="GaKni_Rotation_Mech\\Character\\Back", sortPriority=150}
    Alpha=@n122.B
   )) {name=UVs_Rotation, nodeColor=<rgba 0 0.7188 1 1>}
  n124 = ScalarParameter() {defaultValue=1, group=Wrap_GalaxyKnight_Uvs, parameterName=GaKni_Scale_Vehicle, sortPriority=150}
  n125 = ScalarParameter() {defaultValue=1, group=Wrap_GalaxyKnight_Uvs, parameterName=GaKni_Scale_Vehicle_Side, sortPriority=150}
  n126 = MaterialFunctionCall(Float4=NamedRerouteUsage()) {materialFunction=&BreakOutFloat4Components}
  n127 = LinearInterpolate(
   A=LinearInterpolate(
    A=ScalarParameter() {defaultValue=1, group=Wrap_GalaxyKnight_Uvs, parameterName=GaKni_Scale_Weapon, sortPriority=150}
    B=LinearInterpolate(A=@n124, B=@n125, Alpha=@n126.G)
    Alpha=@n126.R
   )
   B=LinearInterpolate(
    A=ScalarParameter() {defaultValue=1, group=Wrap_GalaxyKnight_Uvs, parameterName=GaKni_Scale_Mech, sortPriority=150}
    B=ScalarParameter() {defaultValue=1, group=Wrap_GalaxyKnight_Uvs, parameterName=GaKni_Scale_Character, sortPriority=150}
    Alpha=@n126.A
   )
   Alpha=@n126.B
  )
  n128 = ScalarParameter() {defaultValue=1, group=Wrap_GalaxyKnight_Uvs, parameterName=GaKni_Scale_Backpack, sortPriority=150}
  n129 = MaterialFunctionCall(Float2=NamedRerouteUsage()) {materialFunction=&BreakOutFloat2Components}
  NamedRerouteDeclaration(None=LinearInterpolate(
    A=LinearInterpolate(A=@n127, B=@n128, Alpha=@n129.R)
    B=ScalarParameter() {defaultValue=1, group=Wrap_GalaxyKnight_Uvs, parameterName=GaKni_Scale_Pickaxe, sortPriority=150}
    Alpha=@n129.G
   )) {name=UVs_Scale, nodeColor=<rgba 0 0.7188 1 1>}
  n130 = ScalarParameter() {defaultValue=1, group=Wrap_GalaxyKnight_Uvs, parameterName=GaKni_Scale_Chest_Vehicle, sortPriority=150}
  n131 = ScalarParameter() {defaultValue=1, group=Wrap_GalaxyKnight_Uvs, parameterName=GaKni_Scale_Chest_Vehicle_Side, sortPriority=150}
  n132 = MaterialFunctionCall(Float4=NamedRerouteUsage()) {materialFunction=&BreakOutFloat4Components}
  n133 = LinearInterpolate(
   A=LinearInterpolate(
    A=ScalarParameter() {defaultValue=1, group=Wrap_GalaxyKnight_Uvs, parameterName=GaKni_Scale_Chest_Weapon, sortPriority=150}
    B=LinearInterpolate(A=@n130, B=@n131, Alpha=@n132.G)
    Alpha=@n132.R
   )
   B=LinearInterpolate(
    A=ScalarParameter() {defaultValue=1, group=Wrap_GalaxyKnight_Uvs, parameterName=GaKni_Scale_Chest_Mech, sortPriority=150}
    B=ScalarParameter() {defaultValue=1, group=Wrap_GalaxyKnight_Uvs, parameterName=GaKni_Scale_Chest_Character, sortPriority=150}
    Alpha=@n132.A
   )
   Alpha=@n132.B
  )
  n134 = ScalarParameter() {defaultValue=1, group=Wrap_GalaxyKnight_Uvs, parameterName=GaKni_Scale_Chest_Backpack, sortPriority=150}
  n135 = MaterialFunctionCall(Float2=NamedRerouteUsage()) {materialFunction=&BreakOutFloat2Components}
  NamedRerouteDeclaration(None=LinearInterpolate(
    A=LinearInterpolate(A=@n133, B=@n134, Alpha=@n135.R)
    B=ScalarParameter() {defaultValue=1, group=Wrap_GalaxyKnight_Uvs, parameterName=GaKni_Scale_Chest_Pickaxe, sortPriority=150}
    Alpha=@n135.G
   )) {name=UVs_Scale_Chest, nodeColor=<rgba 0 0.7188 1 1>}
  n136 = VectorParameter() {group=Wrap_GalaxyKnight_Accent, parameterName=GaKni_Accent_BaseColor, defaultValue=<rgba 0 0.003 0.0004 1>}
  n137 = SetMaterialAttributes(
   MaterialAttributes=NamedRerouteUsage()
   BaseColor=@n136.RGB
   Roughness=ScalarParameter() {defaultValue=0.5, group=Wrap_GalaxyKnight_Accent, parameterName=GaKni_Accent_Roughness}
   Specular=ScalarParameter() {defaultValue=0.5, group=Wrap_GalaxyKnight_Accent, parameterName=GaKni_Accent_Spec}
   Metallic=ScalarParameter() {group=Wrap_GalaxyKnight_Accent, parameterName=GaKni_Accent_Metalic}
  ) {attributeSetTypes [
    #2
    #3
    #5
    #4
    #1
    #6
   ]}
  NamedRerouteDeclaration(None=@n137) {name=Mat_Accent, nodeColor=<rgba 0 0.3906 1 1>}
  NamedRerouteDeclaration(None=FunctionInput(Preview=Constant4Vector().RGBA) {bUsePreviewValueAsDefault=true, inputName=IsState, inputType=FunctionInput_Vector4, sortPriority=4}) {name=IsState, nodeColor=<rgba 0.1497 1 0.1633 1>}
  NamedRerouteDeclaration(None=TextureObjectParameter() {group=Wrap_GalaxyKnight, parameterName=GaKni__Tex_Msk2, texture=&T_BlacksRGB}) {name=Tex_Msk1_1, nodeColor=<rgba 1 0.75 0 1>}
  NamedRerouteDeclaration(None=@n67.Normal) {name=Root_Normal, nodeColor=<rgba 0.4062 0 1 1>}
  NamedRerouteUsage()
  NamedRerouteUsage()
  n138 = Multiply(
   A=Add(A=@n94) {constB=-0.5}
   B=ScalarParameter() {defaultValue=0.5, group=Wrap_GalaxyKnight_SubSurface, parameterName=GaKni_ChestGlowRadius_Inner_Temp}
  )
  Power(Base=SmoothStep(Value=MaterialFunctionCall(Vector 2=@n138) {materialFunction=/Engine/Functions/Engine_MaterialFunctions02/Utility/VectorLength}.V2 Length) {constMax=0, constMin=0.5}) {constExponent=5}
  Power(Base=@n15)
  Saturate(None=@n51)
  n139 = Power(Base=@n18.A) {constExponent=10}
  Saturate(None=@n139)
  Saturate(None=@n139)
  ScalarParameter() {defaultValue=2, group=Wrap_GalaxyKnight_Detail_Normal, parameterName=GaKni_Detail_Normal_TileRate}
  Transform(None=NamedRerouteUsage())
  VertexNormalWS()
 ]}
