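# GameBuddies wrap material function that layers an animated game-screen look onto weapons, vehicles, mechs, characters, backpacks and pickaxes behind a static switch.
# IsState inputs pick per-type parameters for the projected UVs and effects.
# The screen combines bouncing flipbook icons with stepped motion, twinkling stars and a grid, plus a pulsing power glow.
# Primary and secondary masks blend the result over the input material, with a separate accent material.
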MaterialFunction /Game/Weapons/WeaponSkins/Wraps/MaterialFunctions/MF_Wrap_GameBud
graph {nodes [
  n1 = NamedRerouteUsage()
  n2 = MaterialFunctionCall(Float2=@n1) {materialFunction=&BreakOutFloat2Components}
  n3 = AppendVector(A=@n2.G, B=OneMinus(None=@n2.R))
  n4 = ScalarParameter() {group=Wrap_GameBuddies_Icons, parameterName=GameBud_Icons_IsRoll, sortPriority=0}
  n5 = NamedRerouteUsage()
  n6 = Constant() {r=1}
  n7 = NamedRerouteUsage()
  n8 = Add(
   A=Multiply(A=LinearInterpolate(A=@n1, B=@n3, Alpha=@n4), B=AppendVector(A=@n5, B=@n6)) {constB=22}
   B=AppendVector(A=ComponentMask(None=@n7) {r=true}, B=Constant())
  )
  n9 = ComponentMask(None=@n8) {r=true}
  n10 = MaterialFunctionCall(Float4=NamedRerouteUsage()) {materialFunction=&BreakOutFloat4Components}
  n11 = LinearInterpolate(
   A=ScalarParameter() {defaultValue=6, group=Wrap_GameBuddies_Icons, parameterName=GameBud_Icons_Count_Weapon, sortPriority=0}
   B=ScalarParameter() {defaultValue=6, group=Wrap_GameBuddies_Icons, parameterName=GameBud_Icons_count, sortPriority=0}
   Alpha=@n10.R
  )
  n12 = Divide(A=Floor(None=Multiply(A=@n9, B=@n11)), B=@n11)
  n13 = Multiply(A=@n12) {constB=0.7}
  n14 = Constant() {r=1}
  n15 = Multiply(A=@n8, B=AppendVector(A=@n11, B=@n14))
  n16 = Constant3Vector() {constant=<rgba 3 5 0 0>}
  n17 = MaterialFunctionCall!(2 Vector=Add(A=Floor(None=@n15), B=ComponentMask(None=@n16) {g=true, r=true}), RandomSeed) {materialFunction=&/MF_Random_2}
  Add(A=@n13, B=@n17.utput) {constB=0.2}
  Constant() {r=1}
  Constant() {r=0.1}
  n18 = TextureSample() {samplerType=SAMPLERTYPE_Masks, texture=/Game/Weapons/FORT_Rifles/Textures/T_Scar_CM}
  n19 = FunctionInput(Preview=@n18.R) {bUsePreviewValueAsDefault=true, inputName=Mask Primary, inputType=FunctionInput_Scalar, sortPriority=24, previewValue=<1 0 0 1>}
  n20 = FunctionInput(Preview=@n18.B) {bUsePreviewValueAsDefault=true, inputName=Mask Secondary, inputType=FunctionInput_Scalar, sortPriority=24, previewValue=<1 0 0 1>}
  n21 = BlendMaterialAttributes(A=NamedRerouteUsage(), B=NamedRerouteUsage(), Alpha=Saturate(None=Add(A=@n19, B=@n20))) {vertexAttributeBlendType=UseA}
  n22 = NamedRerouteUsage()
  n23 = NamedRerouteDeclaration(None=@n20) {name=Msk_Accent, nodeColor=<rgba 0 1 0.7656 1>}
  n24 = TextureSample() {samplerType=SAMPLERTYPE_Normal, texture=/Game/Weapons/FORT_Rifles/T_SCAR_N}
  n25 = Constant3Vector() {constant=<rgba 1 0 0 0>}
  n26 = GetMaterialAttributes(None=FunctionInput(Preview=SetMaterialAttributes!(MaterialAttributes, Normal=@n24.RGB, BaseColor=@n25) {attributeSetTypes [#1, #2]}) {inputName=Material, inputType=FunctionInput_MaterialAttributes}) {attributeGetTypes [
    #2
    #3
    #1
    #4
   ]}
  FunctionOutput(None=StaticSwitchParameter(
    True=BlendMaterialAttributes(A=@n21, B=@n22, Alpha=@n23) {vertexAttributeBlendType=UseA}
    False=NamedRerouteDeclaration(None=@n26.MaterialAttributes) {name=Root_Material, nodeColor=<rgba 0.7812 1 0 1>}
   ) {group=Customization_Wrap, parameterName=Wrap_GameBuddies})
  n27 = VectorParameter() {group=Wrap_GameBuddies, parameterName=GameBud_BackGround_Color, defaultValue=<rgba 0.1512 0.2 0.1138 1>}
  n28 = VectorParameter() {group="Wrap_GameBuddies_Stars\\Speedlines", parameterName=GameBud_Speedlines_Color, defaultValue=<rgba 0.6616 0.875 0.4979 1>}
  LinearInterpolate(A=@n27.RGB, B=@n28.RGB, Alpha=NamedRerouteUsage())
  LinearInterpolate(A=LinearInterpolate(A=@n16, B=Constant3Vector() {constant=<rgba 3 6 0 0>}, Alpha=@n10.R), B=@n16, Alpha=@n10.B)
  n29 = Add(A=NamedRerouteUsage(), B=ScalarParameter() {defaultValue=0.5, group=Wrap_GameBuddies, parameterName=GameBud_Grid_Darkness})
  n30 = VectorParameter() {group="Wrap_GameBuddies_Stars\\Speedlines", parameterName=GameBud_Stars_Color, defaultValue=<rgba 0.7785 0.9323 0.6604 1>}
  n31 = LinearInterpolate(A=@n27.RGB, B=@n30.RGB, Alpha=NamedRerouteUsage())
  n32 = Constant() {r=3}
  n33 = Constant() {r=1}
  n34 = Add(A=AppendVector(A=Constant(), B=ComponentMask(None=@n7) {g=true}), B=Frac(None=@n15)) {constA=1}
  n35 = AppendVector(A=@n14, B=@n11)
  n36 = Time()
  n37 = VectorParameter() {group=Wrap_GameBuddies_Icons, parameterName=GameBud_Icons_Speed_Y_Weapon, sortPriority=0, defaultValue=<rgba 0 0 0 1>}
  n38 = VectorParameter() {group=Wrap_GameBuddies_Icons, parameterName=GameBud_Icons_Speed_Y, sortPriority=0, defaultValue=<rgba 0 5 0 1>}
  n39 = ComponentMask(None=LinearInterpolate(A=@n37.RGB, B=@n38.RGB, Alpha=@n10.R)) {g=true, r=true}
  n40 = Multiply(
   A=Ceil(None=Multiply(A=@n36, B=@n39))
   B=ScalarParameter() {defaultValue=0.05, group=Wrap_GameBuddies_Icons, parameterName=GameBud_Icons_Step_Amount, sortPriority=5}
  )
  n41 = Add(A=AppendVector(A=Constant(), B=@n17.utput), B=@n40)
  n42 = Add(
   A=Multiply(A=AppendVector(A=@n33, B=@n5), B=Multiply(A=@n34, B=@n35)) {constA=2222222252965888}
   B=Multiply(A=Sine(None=@n41), B=NamedRerouteUsage())
  )
  n43 = MaterialFunctionCall(
   "Animation  Phase (0-1)"=@n13
   Number of Rows=@n32
   Number of Columns=@n32
   Texture=NamedRerouteUsage()
   UVs=@n42
   "Clamp Anim (see tooltip)"=StaticBool()
  ) {materialFunction=&FlipBook}
  n44 = Constant() {r=0.5}
  n45 = MaterialFunctionCall(Float2=@n42) {materialFunction=&BreakOutFloat2Components}
  n46 = Constant() {r=0.45}
  n47 = Multiply(A=SphereMask(A=@n44, B=@n45.R, Radius=@n46) {attenuationRadius=0.5}, B=SphereMask(A=@n45.G, B=@n44, Radius=@n46) {attenuationRadius=0.5})
  n48 = SphereMask(A=@n9, B=Constant() {r=0.5}, Radius=Constant() {r=0.5})
  n49 = Multiply(A=Multiply(A=Multiply(A=@n47, B=@n48), B=@n43.Alpha) {constA=1}, B=NamedRerouteUsage())
  n50 = MaterialFunctionCall(Float4=NamedRerouteUsage()) {materialFunction=&BreakOutFloat4Components}
  n51 = MaterialFunctionCall!(
   For_Safe_Zones
   Normal_Map=NamedRerouteUsage()
   Caculated_Bounds=NamedRerouteUsage()
   Pre_Skinned_Normals=NamedRerouteUsage()
   Is Vehicle=@n50.R
   Is VehicleSide=@n50.R
   Is Mech=@n50.R
   Vehicle_Rotation
   UV_Rotation Bypass=NamedRerouteUsage()
   Mech_Rotation
   Uvs_Offset=NamedRerouteUsage()
   Offset-after rotation
   Uvs_TileRate=NamedRerouteUsage()
  ) {materialFunction=&/MF_Projection_Side_Normal}
  n52 = NamedRerouteDeclaration(None=@n51.UVs) {name=UV_Offset_SD, nodeColor=<rgba 0 0.9531 1 1>}
  n53 = Constant2Vector() {r=-0.25}
  n54 = ComponentMask(None=Multiply(A=Add(A=@n52, B=@n53)) {constB=1.25}) {r=true}
  n55 = OneMinus(None=@n54)
  n56 = Saturate(None=Divide(A=Multiply(A=@n54, B=@n55)) {constB=0.4})
  n57 = Multiply(A=Saturate(None=Add(A=@n56) {constB=0.5}), B=Multiply(A=Saturate(None=@n29), B=LinearInterpolate(A=@n31, B=@n43.Result, Alpha=@n49)))
  n58 = ScalarParameter() {defaultValue=2.5, group=Wrap_GameBuddies, parameterName=GameBud_Screen_Edge_Brightness, sortPriority=5}
  n59 = Multiply(A=Add(A=NamedRerouteUsage()) {constB=-0.5}) {constB=2}
  n60 = NamedRerouteUsage()
  n61 = Multiply(A=LinearInterpolate(A=@n57, B=Multiply(A=@n57, B=@n58) {constB=4}, Alpha=Saturate(None=@n59)), B=Saturate(None=Multiply(A=@n60) {constB=2}))
  n62 = Subtract(A=Time(), B=ScalarParameter() {defaultValue=-9999999, group=Wrap_GameBuddies_reactive, parameterName=GameBud_React_Time_BP})
  n63 = Saturate(None=Divide(A=@n62, B=ScalarParameter() {defaultValue=0.25, group=Wrap_GameBuddies_reactive, parameterName=GameBud_React_Duration_Glow}))
  Multiply(
   A=@n61
   B=LinearInterpolate(
    A=ScalarParameter() {defaultValue=3, group=Wrap_GameBuddies, parameterName=GameBud_Screen_Brightness_Start, sortPriority=5}
    B=ScalarParameter() {defaultValue=0.5, group=Wrap_GameBuddies, parameterName=GameBud_Screen_Brightness, sortPriority=5}
    Alpha=Power(Base=@n63) {constExponent=0.25}
   ) {constA=1}
  )
  n64 = Sine(None=@n40)
  Multiply(A=MaterialFunctionCall!(X range Y range, Gradient Height, ShowGrid?, Thickness, Color, "f(x)"=ComponentMask(None=@n64) {g=true}) {materialFunction=/Engine/Functions/Engine_MaterialFunctions01/Debug/PlotFunctionOnGraph}."f(x) = y")
  n65 = NamedRerouteUsage()
  n66 = TextureSample(UVs=@n65, Tex=NamedRerouteUsage())
  Multiply(B=@n66.RGB) {constA=1}
  n67 = Add(A=Constant2Vector() {g=-0.54, r=-0.039}, B=@n65) {constA=-0.5}
  n68 = OneMinus(None=Subtract(A=MaterialFunctionCall(Vector 2=@n67) {materialFunction=/Engine/Functions/Engine_MaterialFunctions02/Utility/VectorLength}.V2 Length) {constB=0})
  n69 = MaterialFunctionCall(
   A=Constant()
   B=VectorParameter() {group=Wrap_GameBuddies, parameterName=GameBud_Power_Color_A, defaultValue=<rgba 0.0729 0.0013 0 1>}.RGB
   C=VectorParameter() {group=Wrap_GameBuddies, parameterName=GameBud_Power_Color_B, defaultValue=<rgba 1 0.1955 0.0254 1>}.RGB
   Alpha=Power(Base=SmoothStep(Value=@n68) {constMin=0.95}) {constExponent=5}
  ) {materialFunction=&3ColorBlend}
  Multiply(A=@n69.Result)
  n70 = Time()
  n71 = ScalarParameter() {defaultValue=1, group=Wrap_GameBuddies, parameterName=GameBud_Power_Speed}
  n72 = LinearInterpolate(Alpha=Sine(None=Multiply(A=@n70, B=@n71))) {constA=0.75}
  Multiply(A=MaterialFunctionCall("f(x)"=@n72) {materialFunction=&PlotFunctionOnGraph}."f(x) = y")
  Multiply(A=@n12, B=Divide() {constB=6})
  Multiply(A=Frac(None=@n17.utput)) {constB=0.6}
  n73 = Saturate(None=Divide(A=@n62, B=ScalarParameter() {defaultValue=1, group=Wrap_GameBuddies_reactive, parameterName=GameBud_React_Duration}))
  n74 = LinearInterpolate(
   A=ScalarParameter() {defaultValue=0.1, group=Wrap_GameBuddies_reactive, parameterName=GameBud_React___Start}
   B=ScalarParameter() {defaultValue=1, group=Wrap_GameBuddies_reactive, parameterName=GameBud_React___End}
   Alpha=@n73
  )
  Multiply(A=Frac(None=Add(A=@n17.utput, B=@n74) {constB=0.1})) {constB=0.6}
  n75 = ScalarParameter() {defaultValue=2, group="Wrap_GameBuddies_Stars\\Speedlines", parameterName=GameBud_Stars_TileRate}
  n76 = Multiply(A=NamedRerouteUsage(), B=@n75)
  n77 = Multiply(A=@n75) {constB=0.5}
  n78 = Divide(A=Floor(None=Multiply(A=@n76, B=@n77)), B=@n77)
  Multiply(A=@n78)
  n79 = VectorParameter() {group=Wrap_GameBuddies_Accent, parameterName=GameBud_Accent_BaseColor, defaultValue=<rgba 0.1227 0.125 0.06 1>}
  n80 = SetMaterialAttributes!(
   MaterialAttributes=NamedRerouteUsage()
   BaseColor=@n79.RGB
   Roughness=ScalarParameter() {defaultValue=0.6, group=Wrap_GameBuddies_Accent, parameterName=GameBud_Accent_Roughness}
   Specular=ScalarParameter() {defaultValue=0.25, group=Wrap_GameBuddies_Accent, parameterName=GameBud_Accent_Spec}
   Metallic=ScalarParameter() {defaultValue=0.5, group=Wrap_GameBuddies_Accent, parameterName=GameBud_Accent_Metalic}
   Normal
   EmissiveColor
  ) {attributeSetTypes [
    #2
    #3
    #5
    #4
    #1
    #6
   ]}
  NamedRerouteDeclaration(None=@n80) {name=Mat_Accent, nodeColor=<rgba 0 0.3906 1 1>}
  n81 = FunctionInput(Preview=@n18.G) {bUsePreviewValueAsDefault=true, inputName=Mask Crevise, inputType=FunctionInput_Scalar, sortPriority=24, previewValue=<1 0 0 1>}
  NamedRerouteDeclaration(None=@n81) {name=Msk_Curvature, nodeColor=<rgba 0 1 0.7422 1>}
  n82 = ScalarParameter() {group=Wrap_GameBuddies_UVs, parameterName=GameBud_Rotation_Vehicle, sortPriority=150}
  n83 = ScalarParameter() {group=Wrap_GameBuddies_UVs, parameterName=GameBud_Rotation_Vehicle_Side, sortPriority=150}
  n84 = MaterialFunctionCall(Float4=NamedRerouteUsage()) {materialFunction=&BreakOutFloat4Components}
  n85 = LinearInterpolate(
   A=LinearInterpolate(
    A=ScalarParameter() {group=Wrap_GameBuddies_UVs, parameterName=GameBud_Rotation_Weapon, sortPriority=150}
    B=LinearInterpolate(A=@n82, B=@n83, Alpha=@n84.G)
    Alpha=@n84.R
   )
   B=LinearInterpolate(
    A=ScalarParameter() {group=Wrap_GameBuddies_UVs, parameterName=GameBud_Rotation_Mech, sortPriority=150}
    B=ScalarParameter() {group=Wrap_GameBuddies_UVs, parameterName=GameBud_Rotation_Character, sortPriority=150}
    Alpha=@n84.A
   )
   Alpha=@n84.B
  )
  n86 = ScalarParameter() {group=Wrap_GameBuddies_UVs, parameterName=GameBud_Rotation_Backpack, sortPriority=150}
  n87 = MaterialFunctionCall(Float2=NamedRerouteUsage()) {materialFunction=&BreakOutFloat2Components}
  NamedRerouteDeclaration(None=LinearInterpolate(
    A=LinearInterpolate(A=@n85, B=@n86, Alpha=@n87.R)
    B=ScalarParameter() {group=Wrap_GameBuddies_UVs, parameterName=GameBud_Rotation_Pickaxe, sortPriority=150}
    Alpha=@n87.G
   )) {name=UVs_Rotation, nodeColor=<rgba 0 0.7188 1 1>}
  n88 = TextureObjectParameter() {group=Wrap_GameBuddies, parameterName=GameBud__Tex_Icons, sortPriority=3, texture=&T_BlacksRGB}
  NamedRerouteDeclaration(None=@n88) {name=Tex_Icons, nodeColor=<rgba 0.1016 1 0 1>}
  n89 = VectorParameter() {group=Wrap_GameBuddies_UVs, parameterName=GameBud_Offset_Vehicle, defaultValue=<rgba 0 0 0 1>}
  n90 = VectorParameter() {group=Wrap_GameBuddies_UVs, parameterName=GameBud_Offset_VehicleSide, defaultValue=<rgba 0 0 0 1>}
  n91 = MaterialFunctionCall(Float4=NamedRerouteUsage()) {materialFunction=&BreakOutFloat4Components}
  n92 = LinearInterpolate(
   A=LinearInterpolate(
    A=VectorParameter() {group=Wrap_GameBuddies_UVs, parameterName=GameBud_Offset_Weapon, defaultValue=<rgba 0 0 0 1>}.RGB
    B=LinearInterpolate(A=@n89.RGB, B=@n90.RGB, Alpha=@n91.G)
    Alpha=@n91.R
   )
   B=LinearInterpolate(
    A=VectorParameter() {group=Wrap_GameBuddies_UVs, parameterName=GameBud_Offset_Mech, defaultValue=<rgba 0 0 0 1>}.RGB
    B=VectorParameter() {group=Wrap_GameBuddies_UVs, parameterName=GameBud_Offset_Character, defaultValue=<rgba 0 0 0 1>}.RGB
    Alpha=@n91.A
   )
   Alpha=@n91.B
  )
  n93 = VectorParameter() {group=Wrap_GameBuddies_UVs, parameterName=GameBud_Offset_Backpack, defaultValue=<rgba 0 0 0 1>}
  n94 = MaterialFunctionCall(Float2=NamedRerouteUsage()) {materialFunction=&BreakOutFloat2Components}
  NamedRerouteDeclaration(None=LinearInterpolate(
    A=LinearInterpolate(A=@n92, B=@n93.RGB, Alpha=@n94.R)
    B=VectorParameter() {group=Wrap_GameBuddies_UVs, parameterName=GameBud_Offset_Pickaxe, defaultValue=<rgba 0 0 0 1>}.RGB
    Alpha=@n94.G
   )) {name=UVs_Offset, nodeColor=<rgba 0 0.625 1 1>}
  n95 = TextureObjectParameter() {group=Wrap_GameBuddies, parameterName=GameBud__Tex_BC, sortPriority=3, texture=&T_BlacksRGB}
  NamedRerouteDeclaration(None=@n95) {name=Tex_BC, nodeColor=<rgba 0.1016 1 0 1>}
  n96 = TextureObjectParameter() {
   group=Wrap_GameBuddies
   parameterName=GameBud__Tex_N
   samplerType=SAMPLERTYPE_Normal
   sortPriority=3
   texture=&FlatNormalBlue
  }
  NamedRerouteDeclaration(None=@n96) {name=Tex_N, nodeColor=<rgba 0.1016 1 0 1>}
  NamedRerouteDeclaration(None=@n51.UVs No Offsets/Tiling) {name=Uvs_Raw, nodeColor=<rgba 1 0.7969 0 1>}
  n97 = ScalarParameter() {defaultValue=40, group=Wrap_GameBuddies, parameterName=GameBud_Grid_TileRate}
  n98 = ScalarParameter() {defaultValue=80, group=Wrap_GameBuddies, parameterName=GameBud_Grid_TileRate_Character}
  n99 = MaterialFunctionCall(Float4=NamedRerouteUsage()) {materialFunction=&BreakOutFloat4Components}
  n100 = MaterialFunctionCall(Float2=NamedRerouteUsage()) {materialFunction=&BreakOutFloat2Components}
  n101 = LinearInterpolate(
   A=LinearInterpolate(
    A=LinearInterpolate(A=@n97, B=@n98, Alpha=@n99.A)
    B=ScalarParameter() {defaultValue=80, group=Wrap_GameBuddies, parameterName=GameBud_Grid_TileRate_Mech}
    Alpha=@n99.B
   )
   B=ScalarParameter() {defaultValue=20, group=Wrap_GameBuddies, parameterName="GameBud_Grid_TileRate_Back\\Pickaxe"}
   Alpha=Max(A=@n100.R, B=@n100.R)
  )
  n102 = LinearInterpolate(A=NamedRerouteUsage(), B=TextureCoordinate(), Alpha=@n99.A)
  NamedRerouteDeclaration(None=TextureSample(UVs=Multiply(A=@n102, B=@n101), Tex=NamedRerouteUsage()) {samplerType=SAMPLERTYPE_Normal}.R) {name=Grid_Pattern, nodeColor=<rgba 0 1 0.1094 1>}
  n103 = ScalarParameter() {defaultValue=0.5, group=Wrap_GameBuddies, parameterName=GameBud_Screen_Emissive, sortPriority=5}
  n104 = NamedRerouteUsage()
  n105 = ScalarParameter() {group=Wrap_GameBuddies, parameterName=GameBud_Roughness_High, sortPriority=5}
  n106 = ScalarParameter() {defaultValue=1, group=Wrap_GameBuddies, parameterName=GameBud_Roughness_Low, sortPriority=5}
  n107 = NamedRerouteUsage()
  n108 = Saturate(None=@n72)
  n109 = LinearInterpolate(B=@n61, Alpha=@n103)
  n110 = SetMaterialAttributes(
   BaseColor=LinearInterpolate(A=LinearInterpolate(A=@n61, Alpha=@n103) {constB=0}, B=@n66.RGB, Alpha=@n104)
   Roughness=LinearInterpolate(
    A=ScalarParameter() {defaultValue=0.15, group=Wrap_GameBuddies, parameterName=GameBud_Roughness_Screen, sortPriority=5}
    B=LinearInterpolate(A=@n105, B=@n106, Alpha=@n66.A)
    Alpha=NamedRerouteUsage()
   )
   Specular=LinearInterpolate(
    A=ScalarParameter() {defaultValue=0.25, group=Wrap_GameBuddies, parameterName=GameBud_Spec_Screen, sortPriority=5}
    B=ScalarParameter() {defaultValue=0.5, group=Wrap_GameBuddies, parameterName=GameBud_Spec, sortPriority=5}
    Alpha=NamedRerouteUsage()
   )
   Metallic=ScalarParameter() {group=Wrap_GameBuddies, parameterName=GameBud_Metalic, sortPriority=5}
   Normal=Add(A=Multiply(A=@n104, B=@n107), B=NamedRerouteUsage())
   EmissiveColor=Add(A=Multiply(A=@n69.Result, B=@n108), B=LinearInterpolate(A=@n109, Alpha=@n104) {constB=0})
  ) {attributeSetTypes [
    #2
    #3
    #5
    #4
    #1
    #6
   ]}
  NamedRerouteDeclaration(None=@n110) {name=Mat_Base, nodeColor=<rgba 0 0.3906 1 1>}
  n111 = TextureObjectParameter() {group=Wrap_GameBuddies, parameterName=GameBud__Tex_Msk, sortPriority=3, texture=&T_BlacksRGB}
  NamedRerouteDeclaration(None=@n111) {name=Tex_Msk, nodeColor=<rgba 0.1016 1 0 1>}
  NamedRerouteDeclaration(None=@n51.Raw_Normal) {name=Normal_output, nodeColor=<rgba 0 1 0.9531 1>}
  NamedRerouteDeclaration(None=Step(Y=ComponentMask(None=@n65) {r=true}, X=Constant() {r=0.2275})) {name=Screen_Msk, nodeColor=<rgba 0 1 0.4375 1>}
  n112 = ScalarParameter() {defaultValue=1, group=Wrap_GameBuddies_UVs, parameterName=GameBud_Scale_Vehicle, sortPriority=150}
  n113 = ScalarParameter() {defaultValue=1, group=Wrap_GameBuddies_UVs, parameterName=GameBud_Scale_Vehicle_Side, sortPriority=150}
  n114 = MaterialFunctionCall(Float4=NamedRerouteUsage()) {materialFunction=&BreakOutFloat4Components}
  n115 = LinearInterpolate(
   A=LinearInterpolate(
    A=ScalarParameter() {defaultValue=1, group=Wrap_GameBuddies_UVs, parameterName=GameBud_Scale_Weapon, sortPriority=150}
    B=LinearInterpolate(A=@n112, B=@n113, Alpha=@n114.G)
    Alpha=@n114.R
   )
   B=LinearInterpolate(
    A=ScalarParameter() {defaultValue=1, group=Wrap_GameBuddies_UVs, parameterName=GameBud_Scale_Mech, sortPriority=150}
    B=ScalarParameter() {defaultValue=1, group=Wrap_GameBuddies_UVs, parameterName=GameBud_Scale_Character, sortPriority=150}
    Alpha=@n114.A
   )
   Alpha=@n114.B
  )
  n116 = ScalarParameter() {defaultValue=1, group=Wrap_GameBuddies_UVs, parameterName=GameBud_Scale_Backpack, sortPriority=150}
  n117 = MaterialFunctionCall(Float2=NamedRerouteUsage()) {materialFunction=&BreakOutFloat2Components}
  NamedRerouteDeclaration(None=LinearInterpolate(
    A=LinearInterpolate(A=@n115, B=@n116, Alpha=@n117.R)
    B=ScalarParameter() {defaultValue=1, group=Wrap_GameBuddies_UVs, parameterName=GameBud_Scale_Pickaxe, sortPriority=150}
    Alpha=@n117.G
   )) {name=UVs_Scale, nodeColor=<rgba 0 0.7188 1 1>}
  n118 = MaterialFunctionCall(Color=Abs(None=NamedRerouteUsage())) {materialFunction=&SplitComponents}
  n119 = MaterialFunctionCall(Float4=NamedRerouteUsage()) {materialFunction=&BreakOutFloat4Components}
  n120 = LinearInterpolate(A=@n118.R, B=LinearInterpolate(A=@n118.B, B=@n118.B, Alpha=@n119.G), Alpha=@n119.R)
  n121 = ScalarParameter() {defaultValue=1, group=Wrap_GameBuddies, parameterName=GameBud_Tri_Blend_Contrast, sortPriority=0}
  NamedRerouteDeclaration(None=MaterialFunctionCall(In=LinearInterpolate(A=@n120, B=@n118.G, Alpha=@n119.B), Contrast=@n121) {materialFunction=&CheapContrast}.Result) {name=Tri_Mask, nodeColor=<rgba 1 0.4688 0 1>}
  n122 = MaterialFunctionCall(Float3=Abs(None=NamedRerouteUsage())) {materialFunction=&BreakOutFloat3Components}
  n123 = Constant() {r=0.5}
  n124 = NamedRerouteUsage()
  n125 = ComponentMask(None=@n124) {b=true, desc=YZ, g=true}
  n126 = ComponentMask(None=@n124) {b=true, desc=XZ, r=true}
  NamedRerouteDeclaration(None=If(
    A=@n122.B
    B=@n123
    "A > B"=ComponentMask(None=@n124) {desc=YX, g=true, r=true}
    "A < B"=If(A=@n122.R, B=@n123, "A > B"=@n125, "A < B"=@n126)
   )) {name=UVS_Cheap_Tri, nodeColor=<rgba 1 0 0.8438 1>}
  n127 = VectorParameter() {group="Wrap_GameBuddies_Stars\\Speedlines", parameterName=GameBud_Rotation_TS_Weapon, sortPriority=50, defaultValue=<rgba 0 0 -0.25 1>}
  n128 = VectorParameter() {group="Wrap_GameBuddies_Stars\\Speedlines", parameterName=GameBud_Rotation_TS_Vehicle, sortPriority=50, defaultValue=<rgba 0 0 0 1>}
  n129 = VectorParameter() {group="Wrap_GameBuddies_Stars\\Speedlines", parameterName=GameBud_Rotation_TS_Vehicle_Side, sortPriority=50, defaultValue=<rgba 0 0 0 1>}
  n130 = MaterialFunctionCall(Float4=NamedRerouteUsage()) {materialFunction=&BreakOutFloat4Components}
  n131 = VectorParameter() {group="Wrap_GameBuddies_Stars\\Speedlines", parameterName=GameBud_Rotation_TS_Mech, sortPriority=50, defaultValue=<rgba 0.25 0.25 0 1>}
  n132 = VectorParameter() {group="Wrap_GameBuddies_Stars\\Speedlines", parameterName=GameBud_Rotation_TS_Character, sortPriority=50, defaultValue=<rgba 0.25 0.25 0 1>}
  n133 = LinearInterpolate(
   A=LinearInterpolate(A=@n127.RGB, B=LinearInterpolate(A=@n128.RGB, B=@n129.RGB, Alpha=@n130.G), Alpha=@n130.R)
   B=LinearInterpolate(A=@n131.RGB, B=@n132.RGB, Alpha=@n130.A)
   Alpha=@n130.B
  )
  n134 = VectorParameter() {group="Wrap_GameBuddies_Stars\\Speedlines", parameterName=GameBud_Rotation_TS_Backpack, sortPriority=50, defaultValue=<rgba 0 0 0 1>}
  n135 = MaterialFunctionCall(Float2=NamedRerouteUsage()) {materialFunction=&BreakOutFloat2Components}
  n136 = VectorParameter() {group="Wrap_GameBuddies_Stars\\Speedlines", parameterName=GameBud_Rotation_TS_Pickaxe, sortPriority=50, defaultValue=<rgba 0 0 0 1>}
  NamedRerouteDeclaration(None=LinearInterpolate(A=LinearInterpolate(A=@n133, B=@n134.RGB, Alpha=@n135.R), B=@n136.RGB, Alpha=@n135.G)) {name=UVs_TS_Rotation, nodeColor=<rgba 0 0.6016 1 1>}
  NamedRerouteDeclaration(None=@n121) {name=Tri_Contrast, nodeColor=<rgba 1 0.4922 0 1>}
  n137 = VectorParameter() {
   group="Wrap_GameBuddies_Stars\\Speedlines"
   parameterName=GameBud_SpeedLines_Speed_Weapon
   sortPriority=50
   defaultValue=<rgba 0.01 0.5 0.01 1>
  }
  n138 = VectorParameter() {group="Wrap_GameBuddies_Stars\\Speedlines", parameterName=GameBud_SpeedLines_Speed_Vehicle, sortPriority=50, defaultValue=<rgba 0 0 0 1>}
  n139 = VectorParameter() {group="Wrap_GameBuddies_Stars\\Speedlines", parameterName=GameBud_SpeedLines_Speed_VehicleSide, sortPriority=50, defaultValue=<rgba 0 0 0 1>}
  n140 = MaterialFunctionCall(Float4=NamedRerouteUsage()) {materialFunction=&BreakOutFloat4Components}
  n141 = VectorParameter() {group="Wrap_GameBuddies_Stars\\Speedlines", parameterName=GameBud_SpeedLines_Speed_Mech, sortPriority=50, defaultValue=<rgba 0 0 0 1>}
  n142 = VectorParameter() {group="Wrap_GameBuddies_Stars\\Speedlines", parameterName=GameBud_SpeedLines_Speed_Character, sortPriority=50, defaultValue=<rgba 0 0 0 1>}
  n143 = LinearInterpolate(
   A=LinearInterpolate(A=@n137.RGB, B=LinearInterpolate(A=@n138.RGB, B=@n139.RGB, Alpha=@n140.G), Alpha=@n140.R)
   B=LinearInterpolate(A=@n141.RGB, B=@n142.RGB, Alpha=@n140.A)
   Alpha=@n140.B
  )
  n144 = VectorParameter() {group="Wrap_GameBuddies_Stars\\Speedlines", parameterName=GameBud_SpeedLines_Speed_Backpack, sortPriority=50, defaultValue=<rgba 0 0 0 1>}
  n145 = MaterialFunctionCall(Float2=NamedRerouteUsage()) {materialFunction=&BreakOutFloat2Components}
  n146 = VectorParameter() {group="Wrap_GameBuddies_Stars\\Speedlines", parameterName=GameBud_SpeedLines_Speed_Pickaxe, sortPriority=50, defaultValue=<rgba 0 0 0 1>}
  NamedRerouteDeclaration(None=LinearInterpolate(A=LinearInterpolate(A=@n143, B=@n144.RGB, Alpha=@n145.R), B=@n146.RGB, Alpha=@n145.G)) {name=UVs_Speedlines_Speed, nodeColor=<rgba 0 0.625 1 1>}
  n147 = MaterialFunctionCall(Float4=NamedRerouteUsage()) {materialFunction=&BreakOutFloat4Components}
  n148 = Multiply(A=Time(), B=NamedRerouteUsage())
  n149 = Multiply(
   A=NamedRerouteUsage()
   B=ScalarParameter() {defaultValue=2, group="Wrap_GameBuddies_Stars\\Speedlines", parameterName=GameBud_SpeedLines_TileRate}
  )
  NamedRerouteDeclaration(None=MaterialFunctionCall!(
    IsCharacter=@n147.A
    IsStatsB=NamedRerouteUsage()
    Texture=NamedRerouteUsage()
    Rotation_Axis=NamedRerouteUsage()
    Caculated_Bounds=Add(A=@n148, B=@n149)
    Pre_Skinned_Normals=NamedRerouteUsage()
    Blend_Contrast=NamedRerouteUsage()
    Fill_alternate_Axis_Opacity=Constant() {r=1}
    "Fill_alternate_Axis_Color_A=Roughness"=Constant4Vector().RGBA
    Is_Vehicle=@n147.A
    Is_Mech=@n147.A
    Enable_Fill_Z_Weapon
    Fill_Z_Weapon_opacity
    Fill_Z_Weapon_Color
   ) {materialFunction=&/MF_TopsSideProjection_Base}.Alpha) {name=SpeedLines, nodeColor=<rgba 0 0.2969 1 1>}
  n150 = MaterialFunctionCall(Float4=NamedRerouteUsage()) {materialFunction=&BreakOutFloat4Components}
  n151 = MaterialFunctionCall(
   IsCharacter=@n150.A
   IsStatsB=NamedRerouteUsage()
   Texture=NamedRerouteUsage()
   Rotation_Axis=NamedRerouteUsage()
   Caculated_Bounds=@n76
   Pre_Skinned_Normals=NamedRerouteUsage()
   Blend_Contrast=NamedRerouteUsage()
   Fill_alternate_Axis_Opacity=Constant() {r=1}
   "Fill_alternate_Axis_Color_A=Roughness"=Constant4Vector().RGBA
   Is_Vehicle=@n150.A
   Is_Mech=@n150.A
  ) {materialFunction=&MF_TopsSideProjection_Base}
  n152 = ComponentMask(None=@n151.RGB) {r=true}
  n153 = MaterialFunctionCall(Float3=@n78) {materialFunction=&BreakOutFloat3Components}
  n154 = Add(A=Multiply(B=@n153.R) {constA=0.3}, B=Multiply(B=@n153.G) {constA=0.2})
  n155 = Multiply(B=@n153.B) {constA=0.1}
  n156 = Add(
   A=Multiply(A=Time(), B=ScalarParameter() {defaultValue=0.4, group="Wrap_GameBuddies_Stars\\Speedlines", parameterName=GameBud_Stars_Twinkle_Speed})
   B=Add(A=@n152, B=Add(A=@n154, B=@n155))
  )
  n157 = Multiply(A=Multiply(A=Sine(None=@n156), B=Step(X=@n152) {constX=0.5, constY=0.01}), B=@n151.Alpha)
  n158 = Multiply(A=@n157, B=Add(A=@n157, B=@n151.Alpha))
  n159 = Constant() {r=1}
  NamedRerouteDeclaration(None=Saturate(None=MaterialFunctionCall(In=@n158, Contrast=@n159) {materialFunction=&CheapContrast}.Result)) {name=Stars, nodeColor=<rgba 0.8984 0 1 1>}
  n160 = ScalarParameter() {defaultValue=1, group=Wrap_GameBuddies_Icons, parameterName=GameBud_Icons_Bounce_Vehicle, sortPriority=150}
  n161 = ScalarParameter() {defaultValue=1, group=Wrap_GameBuddies_Icons, parameterName=GameBud_Icons_Bounce_Vehicle_Side, sortPriority=150}
  n162 = MaterialFunctionCall(Float4=NamedRerouteUsage()) {materialFunction=&BreakOutFloat4Components}
  n163 = LinearInterpolate(
   A=LinearInterpolate(
    A=ScalarParameter() {defaultValue=0.3, group=Wrap_GameBuddies_Icons, parameterName=GameBud_Icons_Bounce_Weapon, sortPriority=150}
    B=LinearInterpolate(A=@n160, B=@n161, Alpha=@n162.G)
    Alpha=@n162.R
   )
   B=LinearInterpolate(
    A=ScalarParameter() {defaultValue=1, group=Wrap_GameBuddies_Icons, parameterName=GameBud_Icons_Bounce_Mech, sortPriority=150}
    B=ScalarParameter() {defaultValue=1, group=Wrap_GameBuddies_Icons, parameterName=GameBud_Icons_Bounce_Character, sortPriority=150}
    Alpha=@n162.A
   )
   Alpha=@n162.B
  )
  n164 = ScalarParameter() {defaultValue=1, group=Wrap_GameBuddies_Icons, parameterName=GameBud_Icons_Bounce_Backpack, sortPriority=150}
  n165 = MaterialFunctionCall(Float2=NamedRerouteUsage()) {materialFunction=&BreakOutFloat2Components}
  NamedRerouteDeclaration(None=LinearInterpolate(
    A=LinearInterpolate(A=@n163, B=@n164, Alpha=@n165.R)
    B=ScalarParameter() {defaultValue=1, group=Wrap_GameBuddies_Icons, parameterName=GameBud_Icons_Bounce_Pickaxe, sortPriority=150}
    Alpha=@n165.G
   )) {name=Icon_Bounce_Range, nodeColor=<rgba 0 0.7188 1 1>}
  n166 = VectorParameter() {group=Wrap_GameBuddies_Icons, parameterName=GameBud_Icon_Offset_Weapon, defaultValue=<rgba -0.28 -0.45 0 1>}
  n167 = VectorParameter() {group=Wrap_GameBuddies_Icons, parameterName=GameBud_Icon_Offset_Vehicle, defaultValue=<rgba -0.26 -0.5 0 1>}
  n168 = VectorParameter() {group=Wrap_GameBuddies_Icons, parameterName=GameBud_Icon_Offset_VehicleSide, defaultValue=<rgba -0.26 -0.5 0 1>}
  n169 = MaterialFunctionCall(Float4=NamedRerouteUsage()) {materialFunction=&BreakOutFloat4Components}
  n170 = VectorParameter() {group=Wrap_GameBuddies_Icons, parameterName=GameBud_Icon_Offset_Mech, defaultValue=<rgba -0.26 -0.5 0 1>}
  n171 = VectorParameter() {group=Wrap_GameBuddies_Icons, parameterName=GameBud_Icon_Offset_Character, defaultValue=<rgba -0.26 -0.5 0 1>}
  n172 = LinearInterpolate(
   A=LinearInterpolate(A=@n166.RGB, B=LinearInterpolate(A=@n167.RGB, B=@n168.RGB, Alpha=@n169.G), Alpha=@n169.R)
   B=LinearInterpolate(A=@n170.RGB, B=@n171.RGB, Alpha=@n169.A)
   Alpha=@n169.B
  )
  n173 = VectorParameter() {group=Wrap_GameBuddies_Icons, parameterName=GameBud_Icon_Offset_Backpack, defaultValue=<rgba -0.26 -0.5 0 1>}
  n174 = MaterialFunctionCall(Float2=NamedRerouteUsage()) {materialFunction=&BreakOutFloat2Components}
  n175 = VectorParameter() {group=Wrap_GameBuddies_Icons, parameterName=GameBud_Icon_Offset_Pickaxe, defaultValue=<rgba -0.26 -0.5 0 1>}
  NamedRerouteDeclaration(None=LinearInterpolate(A=LinearInterpolate(A=@n172, B=@n173.RGB, Alpha=@n174.R), B=@n175.RGB, Alpha=@n174.G)) {name=UVs_Icon_Offset, nodeColor=<rgba 0 0.625 1 1>}
  n176 = ScalarParameter() {defaultValue=1.25, group=Wrap_GameBuddies_Icons, parameterName=GameBud_Icon_Scale_Vehicle, sortPriority=150}
  n177 = ScalarParameter() {defaultValue=1.25, group=Wrap_GameBuddies_Icons, parameterName=GameBud_Icon_Scale_Vehicle_Side, sortPriority=150}
  n178 = MaterialFunctionCall(Float4=NamedRerouteUsage()) {materialFunction=&BreakOutFloat4Components}
  n179 = LinearInterpolate(
   A=LinearInterpolate(
    A=ScalarParameter() {defaultValue=1.3, group=Wrap_GameBuddies_Icons, parameterName=GameBud_Icon_Scale_Weapon, sortPriority=150}
    B=LinearInterpolate(A=@n176, B=@n177, Alpha=@n178.G)
    Alpha=@n178.R
   )
   B=LinearInterpolate(
    A=ScalarParameter() {defaultValue=1.25, group=Wrap_GameBuddies_Icons, parameterName=GameBud_Icon_Scale_Mech, sortPriority=150}
    B=ScalarParameter() {defaultValue=1.25, group=Wrap_GameBuddies_Icons, parameterName=GameBud_Icon_Scale_Character, sortPriority=150}
    Alpha=@n178.A
   )
   Alpha=@n178.B
  )
  n180 = ScalarParameter() {defaultValue=1.25, group=Wrap_GameBuddies_Icons, parameterName=GameBud_Icon_Scale_Backpack, sortPriority=150}
  n181 = MaterialFunctionCall(Float2=NamedRerouteUsage()) {materialFunction=&BreakOutFloat2Components}
  NamedRerouteDeclaration(None=LinearInterpolate(
    A=LinearInterpolate(A=@n179, B=@n180, Alpha=@n181.R)
    B=ScalarParameter() {defaultValue=1.25, group=Wrap_GameBuddies_Icons, parameterName=GameBud_Icon_Scale_Pickaxe, sortPriority=150}
    Alpha=@n181.G
   )) {name=UVs_Icon_Scale, nodeColor=<rgba 0 0.7188 1 1>}
  NamedRerouteDeclaration(None=FunctionInput(Preview=Constant4Vector() {constant=<rgba 1 0 0 0>}.RGBA) {bUsePreviewValueAsDefault=true, inputName=IsState, inputType=FunctionInput_Vector4, sortPriority=4}) {name=IsState, nodeColor=<rgba 0.1497 1 0.1633 1>}
  NamedRerouteDeclaration(None=FunctionInput(Preview=Constant2Vector()) {bUsePreviewValueAsDefault=true, inputName=IsState_B, inputType=FunctionInput_Vector2, sortPriority=4}) {name=IsState_B, nodeColor=<rgba 0.1497 1 0.1633 1>}
  n182 = PreSkinnedLocalBounds()
  n183 = Subtract(A=VertexInterpolator(VS=PreSkinnedPosition()).PS, B=@n182.Min)
  n184 = MaterialFunctionCall(Color=Subtract(A=@n182.Max, B=@n182.Max)) {materialFunction=&SplitComponents}
  n185 = Max(A=Max(A=@n184.R, B=@n184.R), B=@n184.B)
  NamedRerouteDeclaration(None=FunctionInput(Preview=Divide(A=@n183, B=@n185)) {bUsePreviewValueAsDefault=true, inputName=Calcated_Bounds, sortPriority=11}) {name=Caclcated_Bounds, nodeColor=<rgba 0 0.9297 1 1>}
  n186 = TextureObjectParameter() {group=Wrap_GameBuddies, parameterName=GameBud__Tex_Msk2, sortPriority=3, texture=&T_BlacksRGB}
  NamedRerouteDeclaration(None=@n186) {name=Tex_Msk2, nodeColor=<rgba 0.1016 1 0 1>}
  n187 = PreSkinnedNormal()
  NamedRerouteDeclaration(None=FunctionInput(Preview=VertexInterpolator(VS=@n187).PS) {bUsePreviewValueAsDefault=true, inputName=Pre-Skinned Local Normal, sortPriority=10}) {name=PreSkin_Normal, nodeColor=<rgba 0.8281 1 0 1>}
  NamedRerouteDeclaration(None=@n26.Normal) {name=Root_Normal, nodeColor=<rgba 0.4062 0 1 1>}
  Power(Base=@n73, Exp=ScalarParameter() {defaultValue=0.25, group=Wrap_GameBuddies_reactive, parameterName=GameBud_React_Falloff})
  VertexNormalWS()
 ]}
