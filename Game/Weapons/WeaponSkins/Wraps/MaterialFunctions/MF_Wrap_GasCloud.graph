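# Weapon wrap material function that paints an animated gas cloud, with per-state tuning for weapons, vehicles, vehicle sides and mechs.
# Panned triplanar detail and a scrolling warp sample distort side-projected base UVs over pre-skinned bounds.
# The warped texture's red channel drives a time-scrolled curve atlas or texture color lookup, boosted by curvature.
# The resulting color, roughness and time-of-day-scaled emissive blend over the input material through customization masks, behind a static switch.
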MaterialFunction /Game/Weapons/WeaponSkins/Wraps/MaterialFunctions/MF_Wrap_GasCloud
graph {nodes [
  Constant() {r=1}
  n1 = TextureSample() {samplerType=SAMPLERTYPE_Normal, texture=/Game/Weapons/FORT_Rifles/T_SCAR_N}
  n2 = GetMaterialAttributes(None=FunctionInput(Preview=SetMaterialAttributes!(MaterialAttributes, Normal=@n1.RGB) {attributeSetTypes [#1]}) {inputName=Material, inputType=FunctionInput_MaterialAttributes, sortPriority=-20})
  n3 = VectorParameter() {group=Wrap_GasCloud, parameterName=GasC_BS_Color, sortPriority=1, defaultValue=<rgba 0.003 0.003 0.003 1>}
  n4 = ScalarParameter() {defaultValue=0.2, group=Wrap_GasCloud, parameterName=GasC_WarpDetail_Amount}
  n5 = TextureObjectParameter() {group=Wrap_GasCloud, parameterName="GasC_Tex_DetailWarp   A=WarpBase", texture=&T_BlacksRGB}
  n6 = ScalarParameter() {defaultValue=1, group=Wrap_GasCloud, parameterName=GasC_WarpDetail_Tile_Weapon}
  n7 = FunctionInput(Preview=Constant()) {inputName=isState, inputType=FunctionInput_Vector4, sortPriority=5}
  n8 = MaterialFunctionCall(Color=@n7) {materialFunction=&SplitComponents}
  n9 = LinearInterpolate(
   A=ScalarParameter() {defaultValue=1, group=Wrap_GasCloud, parameterName=GasC_WarpDetail_Tile_Vehicle}
   B=ScalarParameter() {defaultValue=1, group=Wrap_GasCloud, parameterName=GasC_WarpDetail_Tile_VehicleSide}
   Alpha=@n8.G
  )
  n10 = VertexInterpolator(VS=PreSkinnedPosition())
  n11 = PreSkinnedLocalBounds()
  n12 = FunctionInput(Preview=Divide(A=Subtract(A=@n10.PS, B=@n11.Min), B=Subtract(A=@n11.Max, B=@n11.Max))) {bUsePreviewValueAsDefault=true, inputName=Calcated_Bounds, sortPriority=11}
  n13 = Multiply(
   A=LinearInterpolate(
    A=LinearInterpolate(A=@n6, B=@n9, Alpha=@n8.R)
    B=ScalarParameter() {defaultValue=1, group=Wrap_GasCloud, parameterName=GasC_WarpDetail_Tile_Mech}
    Alpha=@n8.B
   )
   B=@n12
  )
  n14 = Multiply(A=Time(), B=VectorParameter() {group=Wrap_GasCloud, parameterName=GasC_WarpDetail_Speed, defaultValue=<rgba -0.05 -0.05 0.05 1>}.RGB)
  n15 = Constant3Vector() {constant=<rgba 0 0 0.25 1>}
  n16 = Constant3Vector() {desc=must activate TS_Vehicle_Side_01 UVs, constant=<rgba 0 0.25 0.25 1>}
  n17 = MaterialFunctionCall(Color=LinearInterpolate(A=LinearInterpolate(A=@n15, B=@n16, Alpha=@n8.R), B=Constant3Vector() {constant=<rgba 0 0 0 1>}, Alpha=@n8.B)) {materialFunction=&SplitComponents}
  n18 = VertexInterpolator(VS=PreSkinnedNormal())
  n19 = ComponentMask(None=MaterialFunctionCall(
    Export Float 4=StaticBool() {value=true}
    TextureObject=@n5
    TextureSize=Constant() {r=1}
    WorldPosition=Add(A=@n13, B=@n14)
    World Space Normal=FunctionInput(Preview=@n18.PS) {bUsePreviewValueAsDefault=true, inputName=Pre-Skinned Local Normal, sortPriority=10}
    RotZAxis=@n17.B
    RotXAxis=@n17.B
    RotYAxis=@n17.B
    ProjectionTransitionContrast=ScalarParameter() {defaultValue=1, group=Wrap_GasCloud, parameterName="GasC_WarpDetail_Blend_Contrast (S)"}
   ) {materialFunction=&/WorldAlignedTexture}.XYZ Texture) {r=true}
  n20 = MaterialFunctionCall(Color=@n7) {materialFunction=&SplitComponents}
  n21 = MaterialFunctionCall(Color=@n7) {materialFunction=&SplitComponents}
  n22 = LinearInterpolate(
   A=ScalarParameter() {group=Wrap_GasCloud, parameterName=GasC_Base_UV_Rotation_Weapon}
   B=LinearInterpolate(
    A=ScalarParameter() {group=Wrap_GasCloud, parameterName=GasC_Base_UV_Rotation_Vehicle}
    B=ScalarParameter() {group=Wrap_GasCloud, parameterName=GasC_Base_UV_Rotation_VehicleSide}
    Alpha=@n21.G
   )
   Alpha=@n21.R
  )
  n23 = ScalarParameter() {group=Wrap_GasCloud, parameterName=GasC_Base_UV_Rotation_Mech}
  n24 = Multiply(
   A=MaterialFunctionCall(
    UVs=MaterialFunctionCall!(
     For_Safe_Zones
     Normal_Map
     Caculated_Bounds=@n12
     Pre_Skinned_Normals
     Is Vehicle=@n20.R
     Is VehicleSide=@n20.R
     Is Mech=@n20.R
     Vehicle_Rotation
     UV_Rotation Bypass
     Mech_Rotation
     Uvs_Offset
     Offset-after rotation
     Uvs_TileRate
    ) {materialFunction=&/MF_Projection_Side_Normal}.UVs No Offsets/Tiling
    "Rotation Angle (0-1)"=LinearInterpolate(A=@n22, B=@n23, Alpha=@n21.B)
   ) {materialFunction=&CustomRotator}.Rotated Values
   B=ScalarParameter() {defaultValue=1, group=Wrap_GasCloud, parameterName=GasC_Base_UV_TileRate}
  )
  n25 = MaterialFunctionCall(Color=@n7) {materialFunction=&SplitComponents}
  n26 = LinearInterpolate(
   A=VectorParameter() {group=Wrap_GasCloud, parameterName=GasC_Warp_Speed_Weapon, defaultValue=<rgba 0.2 0.05 0 1>}.RGB
   B=LinearInterpolate(
    A=VectorParameter() {group=Wrap_GasCloud, parameterName=GasC_Warp_Speed_Vehicle, defaultValue=<rgba 0.2 0.05 0 1>}.RGB
    B=VectorParameter() {group=Wrap_GasCloud, parameterName=GasC_Warp_Speed_VehicleSide, defaultValue=<rgba 0.2 0.05 0 1>}.RGB
    Alpha=@n25.G
   )
   Alpha=@n25.R
  )
  n27 = VectorParameter() {group=Wrap_GasCloud, parameterName=GasC_Warp_Speed_Mech, defaultValue=<rgba 0.2 0.05 0 1>}
  n28 = Multiply(A=Time(), B=ComponentMask(None=LinearInterpolate(A=@n26, B=@n27.RGB, Alpha=@n25.B)) {g=true, r=true})
  n29 = Multiply(
   A=TextureSample(UVs=Add(A=@n24, B=@n28), Tex=@n5).A
   B=ScalarParameter() {defaultValue=0.2, group=Wrap_GasCloud, parameterName=GasC_Warp_Amount}
  )
  n30 = VectorParameter() {group=Wrap_GasCloud, parameterName=GasC_Base_Speed_Vehicle, defaultValue=<rgba 0.01 0.03 0 1>}
  n31 = VectorParameter() {group=Wrap_GasCloud, parameterName=GasC_Base_Speed_VehicleSide, defaultValue=<rgba 0.01 0.03 0 1>}
  n32 = MaterialFunctionCall(Color=@n7) {materialFunction=&SplitComponents}
  n33 = LinearInterpolate(
   A=LinearInterpolate(
    A=VectorParameter() {group=Wrap_GasCloud, parameterName=GasC_Base_Speed_Weapon, defaultValue=<rgba 0.01 0.03 0 1>}.RGB
    B=LinearInterpolate(A=@n30.RGB, B=@n31.RGB, Alpha=@n32.G)
    Alpha=@n32.R
   )
   B=VectorParameter() {group=Wrap_GasCloud, parameterName=GasC_Base_Speed_Mech, defaultValue=<rgba 0.01 0.03 0 1>}.RGB
   Alpha=@n32.B
  )
  n34 = Add(A=Add(A=Multiply(A=@n4, B=@n19), B=Add(A=@n24, B=@n29)), B=Multiply(A=Time(), B=ComponentMask(None=@n33) {g=true, r=true}))
  n35 = Add(
   A=Multiply(
    A=TextureSampleParameter2D(UVs=@n34) {group=Wrap_GasCloud, parameterName=GasC_Base__Tex, texture=&T_BlacksRGB}.R
    B=ScalarParameter() {defaultValue=0.5, group=Wrap_GasCloud, parameterName=GasC_Color_TileRate}
   )
   B=Multiply(A=Time(), B=ScalarParameter() {defaultValue=0.05, group=Wrap_GasCloud, parameterName=GasC_Color_Speed})
  )
  n36 = Frac(None=@n35)
  n37 = CurveAtlasRowParameter(CurveTime=@n36) {
   atlas=/Game/Weapons/WeaponSkins/Wraps/Textures/CruveAtlas_Wraps_01
   curve=/Game/Weapons/WeaponSkins/Wraps/Textures/C_GradSimple
   group=Wrap_GasCloud
   parameterName=GasC_ColorCurve
  }
  n38 = MaterialFunctionCall!(Index=ScalarParameter() {group=Wrap_GasCloud, parameterName="GasC_ColorCurve  Index"}, CCinput=@n37, CCinput_Alpha=@n37, UV=@n36) {materialFunction=&/MF_Global_CC_Tex}
  n39 = TextureSampleParameter2D(UVs=@n35) {group=Wrap_GasCloud, parameterName="GasC_Color__Tex  A=Roughness", texture=/Game/Global/Textures/Default/Masks_and_Gradients/3pxRGBImage}
  n40 = StaticBoolParameter() {group=Wrap_GasCloud, parameterName=GasC_Color use ColorCurve}
  n41 = StaticSwitch(True=@n38.3V, False=@n39.RGB, Value=@n40)
  n42 = FunctionInput(Preview=TextureSample() {samplerType=SAMPLERTYPE_Masks, texture=/Game/Weapons/FORT_Rifles/Textures/T_Scar_CM}.G) {bUsePreviewValueAsDefault=true, inputName=Curavture, inputType=FunctionInput_Scalar, sortPriority=24}
  n43 = LinearInterpolate(
   A=@n41
   B=Multiply(A=ScalarParameter() {defaultValue=2, group=Wrap_GasCloud, parameterName=GasC_Color_Edge_Strength}, B=@n41)
   Alpha=Saturate(None=@n42)
  )
  n44 = Multiply(A=ScalarParameter() {defaultValue=1, group=Wrap_GasCloud, parameterName=GasC_Color_Emissive_Amount}, B=@n43)
  n45 = MaterialFunctionCall(Value_Morning=Constant() {r=0.5}, Value_Day=Constant() {r=1}, Value_Evening=Constant() {r=0.5}, Value_Night=Constant() {r=0.25}) {materialFunction=&Scalar_Time_of_Day}
  n46 = SetMaterialAttributes!(
   MaterialAttributes=@n2.MaterialAttributes
   BaseColor=LinearInterpolate(A=@n3.RGB, B=@n43, Alpha=ScalarParameter() {group=Wrap_GasCloud, parameterName="GasC_0=Simple BC  1=Emissive as BaseColor", sortPriority=1})
   Metallic=ScalarParameter() {group=Wrap_GasCloud, parameterName=GasC_BS_Metalic, sortPriority=1}
   Specular=ScalarParameter() {defaultValue=0.5, group=Wrap_GasCloud, parameterName=GasC_BS_Specular, sortPriority=1}
   Roughness=LinearInterpolate(
    A=ScalarParameter() {defaultValue=0.15, group=Wrap_GasCloud, parameterName=GasC_BS_Roughness, sortPriority=1}
    B=ScalarParameter() {defaultValue=0.15, group=Wrap_GasCloud, parameterName=GasC_BS_Roughness_High, sortPriority=1}
    Alpha=StaticSwitch(True=@n38.Alpha, False=@n39.A, Value=@n40)
   )
   EmissiveColor=LinearInterpolate(
    A=Multiply(A=@n44, B=@n45.Result)
    Alpha=ScalarParameter() {group=Wrap_GasCloud, parameterName="GasC_0=Emissive 1=Disable Emissive", sortPriority=1}
   ) {constB=0}
  ) {attributeSetTypes [
    #2
    #3
    #4
    #5
    #6
   ]}
  n47 = FunctionInput() {bUsePreviewValueAsDefault=true, inputName="Customization Mask (_CM B)", inputType=FunctionInput_Scalar, sortPriority=24}
  n48 = ScalarParameter() {group=Wrap_GasCloud, parameterName=GasC_Remove_Customization Mask, sortPriority=-10}
  n49 = NamedRerouteUsage()
  n50 = NamedRerouteUsage()
  n51 = LinearInterpolate(
   A=Saturate(None=Add(A=@n47, B=@n48))
   B=FunctionInput() {bUsePreviewValueAsDefault=true, inputName=Customization Mask B, inputType=FunctionInput_Scalar, sortPriority=24}
   Alpha=Saturate(None=Subtract(A=@n49, B=@n50))
  )
  FunctionOutput(None=StaticSwitchParameter(True=BlendMaterialAttributes(A=@n2.MaterialAttributes, B=@n46, Alpha=@n51) {vertexAttributeBlendType=UseA}, False=@n2.MaterialAttributes) {group=Customization_Wrap, parameterName=Wrap_GasCloud})
  LinearInterpolate(A=@n16, B=@n16, Alpha=@n8.G)
  Multiply(A=Add(A=@n42) {constB=-0.5}) {constB=1.5}
  NamedRerouteDeclaration(None=MaterialFunctionCall(Float4=@n7) {materialFunction=&BreakOutFloat4Components}.A) {name=IsCharacter, nodeColor=<rgba 0.9785 1 0.0024 1>}
  NamedRerouteDeclaration(None=FunctionInput() {bUsePreviewValueAsDefault=true, inputName=isCharAccent, inputType=FunctionInput_Scalar, sortPriority=78}) {name=isCharAccent, nodeColor=<rgba 0 0.9253 1 1>}
 ]}
