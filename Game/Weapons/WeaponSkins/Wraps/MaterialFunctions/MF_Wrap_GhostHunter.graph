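# Material function for the Ghost Hunter wrap on weapons, vehicles and mechs, blended over input material attributes by the customization mask behind a static switch.
# An object-space spherical gradient, chosen by vehicle/mech inputs and distorted by normals and a swirl texture, scrolls and repeats to sample a color curve atlas.
# That color sets base color and tints swirl-warped, world-aligned bubble caustics on pre-skinned coordinates for emissive.
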MaterialFunction /Game/Weapons/WeaponSkins/Wraps/MaterialFunctions/MF_Wrap_GhostHunter
graph {nodes [
  FunctionInput() {inputName=isVehicle_Side, inputType=FunctionInput_Scalar, sortPriority=5}
  n1 = VertexInterpolator(VS=PreSkinnedPosition())
  n2 = PreSkinnedLocalBounds()
  n3 = FunctionInput(Preview=Divide(A=Subtract(A=@n1.PS, B=@n2.Min), B=Subtract(A=@n2.Max, B=@n2.Max))) {bUsePreviewValueAsDefault=true, inputName=Bounds}
  n4 = VectorParameter() {group=Wrap_GhostHunter, parameterName=GH_ObjectSpaceGradStart_Weapon, defaultValue=<rgba 0 0 0 1>}
  n5 = ScalarParameter() {defaultValue=1, group=Wrap_GhostHunter, parameterName=GH_ObjectSpaceGradRadius_Weapon}
  n6 = Constant()
  n7 = VectorParameter() {group=Wrap_GhostHunter, parameterName=GH_ObjectSpaceGradStart_Vehicle, defaultValue=<rgba 0 0 0 1>}
  n8 = ScalarParameter() {defaultValue=1, group=Wrap_GhostHunter, parameterName=GH_ObjectSpaceGradRadius_Vehicle}
  n9 = Constant()
  n10 = FunctionInput() {inputName=isVehicle, inputType=FunctionInput_Scalar, sortPriority=5}
  n11 = FunctionInput() {bUsePreviewValueAsDefault=true, inputName=isMech, inputType=FunctionInput_Scalar, sortPriority=4}
  n12 = LinearInterpolate(
   A=LinearInterpolate(A=SphereMask(A=@n3, B=@n4.RGB, Radius=@n5, Hardness=@n6), B=SphereMask(A=@n3, B=@n7.RGB, Radius=@n8, Hardness=@n9), Alpha=@n10)
   B=SphereMask(
    A=@n3
    B=VectorParameter() {group=Wrap_GhostHunter, parameterName=GH_ObjectSpaceGradStart_Mech, defaultValue=<rgba 0 0 0 1>}.RGB
    Radius=ScalarParameter() {defaultValue=1, group=Wrap_GhostHunter, parameterName=GH_ObjectSpaceGradRadius_Mech}
    Hardness=Constant()
   )
   Alpha=@n11
  )
  n13 = GetMaterialAttributes(None=FunctionInput() {bUsePreviewValueAsDefault=true, inputName=Material, inputType=FunctionInput_MaterialAttributes, sortPriority=-1}) {attributeGetTypes [#1]}
  n14 = Multiply(
   A=ComponentMask(None=@n13.Normal) {b=true}
   B=ScalarParameter() {defaultValue=0.2, group=Wrap_GhostHunter, parameterName=GH_NormalDistortionWeight}
  )
  n15 = Add(
   A=Multiply(A=Add(A=@n12, B=@n14), B=ScalarParameter() {defaultValue=1, group=Wrap_GhostHunter, parameterName=GH_StretchGradient})
   B=Multiply(A=Time(), B=ScalarParameter() {defaultValue=0.1, group=Wrap_GhostHunter, parameterName=GH_speed})
  )
  n16 = VertexInterpolator(VS=PreSkinnedNormal())
  n17 = ComponentMask(None=MaterialFunctionCall(
    TextureObject=TextureObject() {texture=/Game/Global/Textures/Default/Flowmaps/T_SwirlFlow_02_D}
    TextureSize=Constant3Vector() {constant=<rgba 1.75 1.75 1.75 0>}
    WorldPosition=NamedRerouteUsage()
    World Space Normal=@n16.PS
   ) {materialFunction=&WorldAlignedTexture}.XYZ Texture) {r=true}
  n18 = Frac(None=Multiply(A=Add(A=@n15, B=@n17), B=ScalarParameter() {defaultValue=1.6, group=Wrap_GhostHunter, parameterName=GH_Repetition}))
  n19 = CurveAtlasRowParameter(CurveTime=@n18) {
   atlas=/Game/Weapons/WeaponSkins/Wraps/Textures/CruveAtlas_Wraps_01
   curve=/Game/Characters/Player/Female/Medium/Bodies/F_MED_Ghost_Hunter_Armor/Materials/T_GhostHunter_Colors_02
   defaultValue=60
   group=Wrap_GhostHunter
   parameterName=GH_ColorCurve
  }
  n20 = MaterialFunctionCall!(
   Index=ScalarParameter() {defaultValue=60, group=Wrap_GhostHunter, parameterName="GH_ColorCurve  Index"}
   CCinput=@n19
   CCinput_Alpha
   UV=@n18
  ) {materialFunction=&/MF_Global_CC_Tex}
  n21 = TextureObject() {
   samplerType=SAMPLERTYPE_LinearGrayscale
   texture=/Game/Characters/Player/Female/Medium/Bodies/F_MED_Ghost_Hunter_Armor/Textures/T_F_MED_Ghost_Hunter_Bubbles
  }
  n22 = ScalarParameter() {defaultValue=1, group=Wrap_GhostHunter, parameterName=GH_CausticScale_Weapon}
  n23 = ScalarParameter() {defaultValue=1, group=Wrap_GhostHunter, parameterName=GH_CausticScale_Vehicle}
  n24 = NamedRerouteUsage()
  n25 = Time()
  n26 = Multiply(
   A=Add(A=@n17, B=Multiply(A=@n25) {constB=-0.02})
   B=ScalarParameter() {defaultValue=0.3, group=Wrap_GhostHunter, parameterName=GH_CausticSwirlStrength}
  )
  n27 = NamedRerouteUsage()
  n28 = Multiply(A=Time(), B=ScalarParameter() {defaultValue=-0.02, group=Wrap_GhostHunter, parameterName=GH_CausticSpeed}) {constB=-0.02}
  n29 = MaterialFunctionCall(
   TextureObject=@n21
   TextureSize=LinearInterpolate(
    A=LinearInterpolate(A=@n22, B=@n23, Alpha=@n24)
    B=ScalarParameter() {defaultValue=1, group=Wrap_GhostHunter, parameterName=GH_CausticScale_Mech}
    Alpha=NamedRerouteUsage()
   )
   WorldPosition=Add(A=ComponentMask(None=@n26) {g=true}, B=Add(A=@n27, B=@n28))
   World Space Normal=NamedRerouteUsage()
  ) {materialFunction=&WorldAlignedTexture}
  n30 = Multiply(A=Power(Base=ComponentMask(None=@n29.XYZ Texture) {r=true}) {constExponent=1}, B=@n20.3V)
  n31 = ScalarParameter() {defaultValue=0.8, group=GhostHunter, parameterName=GH_CausticStrength}
  n32 = SetMaterialAttributes!(
   MaterialAttributes=@n13.MaterialAttributes
   ShadingModel
   BaseColor=Multiply(A=@n20.3V, B=ScalarParameter() {defaultValue=1, group=Wrap_GhostHunter, parameterName=GH_BaseColorBrightness, sortPriority=-10})
   Metallic=ScalarParameter() {defaultValue=1, group=Wrap_GhostHunter, parameterName=GH_Metalness, sortPriority=-10}
   Specular=ScalarParameter() {defaultValue=0.5, group=Wrap_GhostHunter, parameterName=GH_Specular, sortPriority=-10}
   Roughness=ScalarParameter() {defaultValue=0.5, group=Wrap_GhostHunter, parameterName=GH_Roughness, sortPriority=-10}
   Normal=MaterialFunctionCall(Normal=@n13.Normal, Flatness=ScalarParameter() {defaultValue=0.5, group=Wrap_GhostHunter, parameterName=GH_DanceLines_NormalFlattness}) {materialFunction=&FlattenNormal}.Result
   EmissiveColor=Multiply(
    A=Multiply(A=@n30, B=@n31)
    B=ScalarParameter() {defaultValue=0.75, group=Wrap_GhostHunter, parameterName=GH_Emissive_Strenght, sortPriority=-9}
   ) {constB=0.5}
  ) {attributeSetTypes [
    #2
    #3
    #4
    #5
    #6
    #1
    #7
   ]}
  n33 = Saturate(None=Add(
    A=FunctionInput() {bUsePreviewValueAsDefault=true, inputName="Customization Mask (_CM B)", inputType=FunctionInput_Scalar, sortPriority=24}
    B=ScalarParameter() {group=Wrap_GhostHunter, parameterName=GH_Remove_Customization Mask, sortPriority=-10}
   ))
  FunctionOutput(None=StaticSwitchParameter(True=BlendMaterialAttributes(A=@n13.MaterialAttributes, B=@n32, Alpha=@n33), False=@n13.MaterialAttributes) {group=Customization, parameterName=Wrap_GhostHunter})
  Multiply() {constA=1}
  n34 = LinearInterpolate(
   A=VectorParameter() {group=Wrap_GhostHunter, parameterName=GH_Paining_Weapon, defaultValue=<rgba 0 0 0 1>}.RGB
   B=VectorParameter() {group=Wrap_GhostHunter, parameterName=GH_Paining_vehicle, defaultValue=<rgba 0 0 0 1>}.RGB
   Alpha=@n10
  )
  n35 = VectorParameter() {group=Wrap_GhostHunter, parameterName=GH_Paining_Mech, defaultValue=<rgba 0 0 0 1>}
  n36 = Add(A=@n3, B=Multiply(A=Time(), B=LinearInterpolate(A=@n34, B=@n35.RGB, Alpha=@n11)))
  n37 = Multiply(A=@n36, B=ScalarParameter() {defaultValue=75, group=Wrap_GhostHunter, parameterName=GH_Noise_tile})
  Multiply(
   A=ComponentMask(None=Noise!(World Position=@n37, FilterWidth) {bTurbulence=false, levels=1, noiseFunction=NOISEFUNCTION_GradientALU, outputMin=0, scale=0.04}) {g=true}
   B=ScalarParameter() {defaultValue=1, group=Wrap_GhostHunter, parameterName=GH_NoiseWarpIntensity}
  )
  NamedRerouteDeclaration(None=@n36) {name=3D Coords, nodeColor=<rgba 0.1953 1 0 1>}
  NamedRerouteDeclaration(None=@n16.PS) {name=Preskinned Normals, nodeColor=<rgba 0 1 0.1797 1>}
  NamedRerouteDeclaration(None=@n10) {name=isVehicle, nodeColor=<rgba 1 0.375 0 1>}
  NamedRerouteDeclaration(None=@n11) {name=isMech, nodeColor=<rgba 0.1719 0 1 1>}
  Reroute(None=@n36)
  TextureCoordinate()
  TextureSampleParameter2D() {
   group=GhostHunter
   parameterName=GH_Caustics
   samplerType=SAMPLERTYPE_LinearGrayscale
   texture=&T_F_MED_Ghost_Hunter_Bubbles
  }
 ]}
