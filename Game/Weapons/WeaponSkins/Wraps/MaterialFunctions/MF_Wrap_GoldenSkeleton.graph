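# Golden Skeleton weapon wrap material function that, behind a static switch, repaints base color, metallic, roughness, specular and normal through customization masks.
# Side-projected base textures use isState to choose rotation and offset for weapon, vehicle or mech.
# A radial-UV sun with time-scrolled, warped tentacles is composited over them, limited by surface facing.
# A curvature-shaded secondary layer, gold by default, covers mask B plus pattern areas within mask A.
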MaterialFunction /Game/Weapons/WeaponSkins/Wraps/MaterialFunctions/MF_Wrap_GoldenSkeleton
graph {nodes [
  n1 = FunctionInput(Preview=SetMaterialAttributes!(MaterialAttributes, Normal) {attributeSetTypes [#1]}) {bUsePreviewValueAsDefault=true, inputName=Material, inputType=FunctionInput_MaterialAttributes, sortPriority=-20}
  n2 = GetMaterialAttributes(None=@n1) {attributeGetTypes [
    #2
    #3
    #4
    #5
    #1
   ]}
  n3 = VectorParameter() {group=Wrap_GoldenSkeleton, parameterName=GoldSkel_Sun_Color, sortPriority=0, defaultValue=<rgba 0.2307 0.0467 0.0467 1>}
  n4 = TextureObjectParameter() {
   group=Wrap_GoldenSkeleton
   parameterName=GoldSkel_Base__Tex_N
   samplerType=SAMPLERTYPE_Normal
   texture=&FlatNormalBlue
  }
  n5 = FunctionInput() {bUsePreviewValueAsDefault=true, inputName=isState, inputType=FunctionInput_Vector4, sortPriority=5, previewValue=<0 0 0 0>}
  n6 = MaterialFunctionCall(Color=@n5) {materialFunction=&SplitComponents}
  n7 = ScalarParameter() {group=Wrap_GoldenSkeleton, parameterName=GoldSkel_Base_Rotation_Weapon}
  n8 = MaterialFunctionCall(Color=@n5) {materialFunction=&SplitComponents}
  n9 = LinearInterpolate(
   A=ScalarParameter() {group=Wrap_GoldenSkeleton, parameterName=GoldSkel_Base_Rotation_Vehicle}
   B=ScalarParameter() {group=Wrap_GoldenSkeleton, parameterName=GoldSkel_Base_Rotation_VehicleSide}
   Alpha=@n8.G
  )
  n10 = VectorParameter() {group=Wrap_GoldenSkeleton, parameterName=GoldSkel_Base_Offset_Weapon, defaultValue=<rgba 0 0 0 1>}
  n11 = MaterialFunctionCall(Color=@n5) {materialFunction=&SplitComponents}
  n12 = LinearInterpolate(
   A=VectorParameter() {group=Wrap_GoldenSkeleton, parameterName=GoldSkel_Base_Offset_Vehicle, defaultValue=<rgba 0 0 0 1>}.RGB
   B=VectorParameter() {group=Wrap_GoldenSkeleton, parameterName=GoldSkel_Base_Offset_VehicleSide, defaultValue=<rgba 0 0 0 1>}.RGB
   Alpha=@n11.G
  )
  n13 = FunctionInput(Preview=VertexInterpolator(VS=PreSkinnedNormal()).PS) {bUsePreviewValueAsDefault=true, inputName=Pre-Skinned Local Normal, sortPriority=10}
  n14 = PreSkinnedLocalBounds()
  n15 = Subtract(A=VertexInterpolator(VS=PreSkinnedPosition()).PS, B=@n14.Min)
  n16 = Subtract(A=@n14.Max, B=@n14.Max)
  n17 = MaterialFunctionCall!(
   For_Safe_Zones
   Normal_Map=@n4
   Caculated_Bounds=FunctionInput(Preview=Divide(A=@n15, B=@n16)) {bUsePreviewValueAsDefault=true, inputName=Calcated_Bounds, sortPriority=11}
   Pre_Skinned_Normals=@n13
   Is Vehicle=@n6.R
   Is VehicleSide=@n6.R
   Is Mech=@n6.R
   Vehicle_Rotation
   UV_Rotation Bypass=LinearInterpolate(
    A=LinearInterpolate(A=@n7, B=@n9, Alpha=@n8.R)
    B=ScalarParameter() {defaultValue=0.25, group=Wrap_GoldenSkeleton, parameterName=GoldSkel_Base_Rotation_Mech}
    Alpha=@n8.B
   )
   Mech_Rotation
   Uvs_Offset=LinearInterpolate(
    A=LinearInterpolate(A=@n10.RGB, B=@n12, Alpha=@n11.R)
    B=VectorParameter() {group=Wrap_GoldenSkeleton, parameterName=GoldSkel_Base_Offset_Mech, defaultValue=<rgba 0 0 0 1>}.RGB
    Alpha=@n11.B
   )
   Offset-after rotation
   Uvs_TileRate=ScalarParameter() {defaultValue=1, group=Wrap_GoldenSkeleton, parameterName=GoldSkel_Base_Scale}
  ) {materialFunction=&/MF_Projection_Side_Normal}
  n18 = VectorParameter() {group=Wrap_GoldenSkeleton, parameterName=GoldSkel_Sun_Offset_Weapon, defaultValue=<rgba -0.5 0.5 0 1>}
  n19 = MaterialFunctionCall(Float4=@n5) {materialFunction=&BreakOutFloat4Components}
  n20 = LinearInterpolate(
   A=VectorParameter() {group=Wrap_GoldenSkeleton, parameterName=GoldSkel_Sun_Offset_Vehicle, defaultValue=<rgba 0 0 0 1>}.RGB
   B=VectorParameter() {group=Wrap_GoldenSkeleton, parameterName=GoldSkel_Sun_Offset_VehicleSide, defaultValue=<rgba 0 0 0 1>}.RGB
   Alpha=@n19.G
  )
  n21 = VectorParameter() {group=Wrap_GoldenSkeleton, parameterName=GoldSkel_Sun_Offset_Mech, defaultValue=<rgba 0 0 0 1>}
  n22 = VectorParameter() {group=Wrap_GoldenSkeleton, parameterName=GoldSkel_Sun_Offset_Character, defaultValue=<rgba 0 0 0 1>}
  n23 = ComponentMask(None=LinearInterpolate(A=LinearInterpolate(A=@n18.RGB, B=@n20, Alpha=@n19.R), B=LinearInterpolate(A=@n21.RGB, B=@n22.RGB, Alpha=@n19.A), Alpha=@n19.B)) {g=true, r=true}
  n24 = MaterialFunctionCall(Float4=@n5) {materialFunction=&BreakOutFloat4Components}
  n25 = LinearInterpolate(
   A=ScalarParameter() {defaultValue=1, group=Wrap_GoldenSkeleton, parameterName=GoldSkel_Sun_Scale_Weapon}
   B=LinearInterpolate(
    A=ScalarParameter() {defaultValue=1, group=Wrap_GoldenSkeleton, parameterName=GoldSkel_Sun_Scale_Vehicle}
    B=ScalarParameter() {defaultValue=1, group=Wrap_GoldenSkeleton, parameterName=GoldSkel_Sun_Scale_VehicleSide}
    Alpha=@n24.G
   )
   Alpha=@n24.R
  )
  n26 = LinearInterpolate(
   A=ScalarParameter() {defaultValue=1, group=Wrap_GoldenSkeleton, parameterName=GoldSkel_Sun_Scale_Mech}
   B=ScalarParameter() {defaultValue=1, group=Wrap_GoldenSkeleton, parameterName=GoldSkel_Sun_Scale_Character}
   Alpha=@n24.A
  )
  n27 = MaterialFunctionCall(Vector or UVs=Multiply(A=Add(A=@n17.UVs No Offsets/Tiling, B=@n23), B=LinearInterpolate(A=@n25, B=@n26, Alpha=@n24.B))) {materialFunction=&VectorToRadialValue}
  n28 = VectorParameter() {group=Wrap_GoldenSkeleton, parameterName=GoldSkel_SunInner_TileRate, defaultValue=<rgba 1 1 0 1>}
  n29 = MaterialFunctionCall!(
   UVs=@n17.UVs No Offsets/Tiling
   Texture Size=ScalarParameter() {defaultValue=256, group=Wrap_GoldenSkeleton, parameterName=GoldSkel_Sun_Miplevel _Size}
  ) {materialFunction=/Engine/Functions/Engine_MaterialFunctions01/Texturing/ComputeMipLevel}
  n30 = TextureSample!(
   UVs=Multiply(A=@n27.Radial Coordinates, B=ComponentMask(None=@n28.RGB) {g=true, r=true})
   Tex=TextureObjectParameter() {group=Wrap_GoldenSkeleton, parameterName=GoldSkel_Sun__Tex_M, texture=&T_BlacksRGB}
   Level=@n29.MipLevel
   Apply View MipBias
  ) {mipValueMode=TMVM_MipLevel}
  n31 = ScalarParameter() {defaultValue=20, group=Wrap_GoldenSkeleton, parameterName=GoldSkel_Sun_Contrast}
  n32 = Multiply(
   A=@n27.Radial Coordinates
   B=ComponentMask(None=VectorParameter() {group=Wrap_GoldenSkeleton, parameterName=GoldSkel_Sun_Tentical_TileRate, defaultValue=<rgba 4 2.5 0 1>}.RGB) {g=true, r=true}
  )
  n33 = ComponentMask(None=Multiply(A=Time(), B=VectorParameter() {group=Wrap_GoldenSkeleton, parameterName=GoldSkel_Sun_Tentical_Speed, defaultValue=<rgba 0 -0.1 0 0>}.RGB)) {g=true, r=true}
  n34 = VectorParameter() {group=Wrap_GoldenSkeleton, parameterName=GoldSkel_Sun_Tentical_Warp_Scale, defaultValue=<rgba 1 1.25 0 1>}
  n35 = Time()
  n36 = VectorParameter() {group=Wrap_GoldenSkeleton, parameterName=GoldSkel_Sun_Tentical_Speed_Warp, defaultValue=<rgba 0 -0.1 0 0>}
  n37 = Add(
   A=Multiply(A=@n27.Radial Coordinates, B=ComponentMask(None=@n34.RGB) {g=true, r=true})
   B=ComponentMask(None=Multiply(A=@n35, B=@n36.RGB)) {g=true, r=true}
  )
  n38 = TextureObjectParameter() {group=Wrap_GoldenSkeleton, parameterName=GoldSkel_Sun__Tex_M1, texture=&T_BlacksRGB}
  n39 = Multiply(
   A=Add(A=TextureSample(UVs=@n37, Tex=@n38, Level=@n29.MipLevel) {mipValueMode=TMVM_MipLevel}.A) {constB=-0.5}
   B=ScalarParameter() {defaultValue=0.1, group=Wrap_GoldenSkeleton, parameterName=GoldSkel_Sun_Tentical_Warp_Amount}
  )
  n40 = TextureSample(UVs=Add(A=Add(A=@n32, B=@n33), B=Multiply(A=@n30.G, B=@n39)), Tex=@n38, Level=@n29.MipLevel) {mipValueMode=TMVM_MipLevel}
  n41 = Saturate(None=Add(
    A=MaterialFunctionCall(In=@n30.A, Contrast=@n31) {materialFunction=&CheapContrast}.Result
    B=MaterialFunctionCall(Transition Phase=@n30.R, Height Texture=@n40.R, Contrast=@n31) {materialFunction=&HeightLerp}.Alpha
   ))
  n42 = Multiply(A=OneMinus(None=@n30.B), B=@n40.G)
  n43 = MaterialFunctionCall(Color=Abs(None=@n13)) {materialFunction=&SplitComponents}
  n44 = MaterialFunctionCall(Color=@n5) {materialFunction=&SplitComponents}
  n45 = LinearInterpolate(A=@n43.R, B=LinearInterpolate(A=@n43.B, B=@n43.B, Alpha=@n44.G), Alpha=@n44.R)
  n46 = Multiply(
   A=Saturate(None=Subtract(A=@n41, B=@n42))
   B=MaterialFunctionCall(
    In=LinearInterpolate(A=@n45, B=@n43.G, Alpha=@n44.B)
    Contrast=ScalarParameter() {defaultValue=20, group=Wrap_GoldenSkeleton, parameterName=GoldSkel_Tri_Contrast, sortPriority=-1}
   ) {materialFunction=&CheapContrast}.Result
  )
  n47 = TextureSample!(
   UVs=@n17.UVs
   Tex=TextureObjectParameter() {group=Wrap_GoldenSkeleton, parameterName=GoldSkel_Base__Tex_BC, texture=&T_BlacksRGB}
   Apply View MipBias
  ) {samplerType=SAMPLERTYPE_LinearColor}
  n48 = LinearInterpolate(A=@n47.RGB, B=@n3.RGB, Alpha=@n46)
  n49 = TextureSample() {samplerType=SAMPLERTYPE_Masks, texture=/Game/Weapons/FORT_Rifles/Textures/T_Scar_CM}
  n50 = FunctionInput(Preview=@n49.R) {bUsePreviewValueAsDefault=true, inputName=Customization Mask A, inputType=FunctionInput_Scalar, sortPriority=24}
  n51 = VectorParameter() {group=Wrap_GoldenSkeleton, parameterName=GoldSkel_2nd_BaseColorA, sortPriority=0, defaultValue=<rgba 0.8021 0.5005 0.0033 1>}
  n52 = VectorParameter() {group=Wrap_GoldenSkeleton, parameterName=GoldSkel_2nd_BaseColor_B, sortPriority=0, defaultValue=<rgba 0.3864 0.2664 0.0648 1>}
  n53 = ScalarParameter() {defaultValue=10, group=Wrap_GoldenSkeleton, parameterName=GoldSkel_2nd_Curv_Mask A, sortPriority=0}
  n54 = ScalarParameter() {defaultValue=-7, group=Wrap_GoldenSkeleton, parameterName=GoldSkel_2nd_Curv_Mask B, sortPriority=0}
  n55 = FunctionInput(Preview=TextureSample() {samplerType=SAMPLERTYPE_Masks, texture=&T_Scar_CM}.G) {bUsePreviewValueAsDefault=true, inputName=Curavture, inputType=FunctionInput_Scalar, sortPriority=24}
  n56 = Saturate(None=OneMinus(None=LinearInterpolate(A=@n53, B=@n54, Alpha=@n55)))
  n57 = TextureSample(
   UVs=@n17.UVs
   Tex=TextureObjectParameter() {group=Wrap_GoldenSkeleton, parameterName=GoldSkel_Base__Tex_M, texture=&T_BlacksRGB}
  ) {samplerType=SAMPLERTYPE_LinearColor}
  n58 = Add(A=@n56, B=Multiply(A=@n50, B=Multiply(A=@n47.A, B=@n57.R)))
  n59 = Saturate(None=Add(
    A=FunctionInput(Preview=@n49.B) {bUsePreviewValueAsDefault=true, inputName=Customization Mask B, inputType=FunctionInput_Scalar, sortPriority=24}
    B=Multiply(A=@n50, B=@n57.R)
   ) {constB=0})
  n60 = LinearInterpolate(
   A=ScalarParameter() {group=Wrap_GoldenSkeleton, parameterName=GoldSkel_Metallic_A, sortPriority=0}
   B=ScalarParameter() {defaultValue=1, group=Wrap_GoldenSkeleton, parameterName=GoldSkel_Sun_Metallic, sortPriority=0}
   Alpha=@n46
  )
  n61 = LinearInterpolate(
   A=LinearInterpolate(
    A=ScalarParameter() {group=Wrap_GoldenSkeleton, parameterName=GoldSkel_Roghness_A, sortPriority=0}
    B=ScalarParameter() {defaultValue=1, group=Wrap_GoldenSkeleton, parameterName=GoldSkel_Roghness_B, sortPriority=0}
    Alpha=@n47.A
   )
   B=ScalarParameter() {defaultValue=0.35, group=Wrap_GoldenSkeleton, parameterName=GoldSkel_Sun_Roghness, sortPriority=0}
   Alpha=@n46
  )
  n62 = ScalarParameter() {defaultValue=0.2, group=Wrap_GoldenSkeleton, parameterName=GoldSkel_2nd_Roghness_A, sortPriority=0}
  n63 = ScalarParameter() {defaultValue=0.5, group=Wrap_GoldenSkeleton, parameterName=GoldSkel_2nd_Roghness_B, sortPriority=0}
  n64 = ScalarParameter() {defaultValue=0.5, group=Wrap_GoldenSkeleton, parameterName=GoldSkel_Specular, sortPriority=0}
  n65 = Add(A=@n2.Normal, B=LinearInterpolate(A=@n17.Normal, Alpha=@n46) {constB=0})
  n66 = ScalarParameter() {defaultValue=1, group=Wrap_GoldenSkeleton, parameterName=GoldSkel_Normal_intensty, sortPriority=0}
  n67 = SetMaterialAttributes!(
   MaterialAttributes=@n2.MaterialAttributes
   BaseColor=LinearInterpolate(A=LinearInterpolate(A=@n2.BaseColor, B=@n48, Alpha=@n50), B=LinearInterpolate(A=@n51.RGB, B=@n52.RGB, Alpha=@n58), Alpha=@n59)
   Metallic=LinearInterpolate(
    A=LinearInterpolate(A=@n2.Metallic, B=@n60, Alpha=@n50)
    B=ScalarParameter() {defaultValue=1, group=Wrap_GoldenSkeleton, parameterName=GoldSkel_2nd_Metallic, sortPriority=0}
    Alpha=@n59
   )
   Roughness=LinearInterpolate(A=LinearInterpolate(A=@n2.Roughness, B=@n61, Alpha=@n50), B=LinearInterpolate(A=@n62, B=@n63, Alpha=@n58), Alpha=@n59)
   Specular=LinearInterpolate(
    A=LinearInterpolate(A=@n2.Specular, B=@n64, Alpha=@n50)
    B=ScalarParameter() {defaultValue=0.5, group=Wrap_GoldenSkeleton, parameterName=GoldSkel_2nd_Specular, sortPriority=0}
    Alpha=@n59
   )
   EmissiveColor
   Normal=LinearInterpolate(A=@n2.Normal, B=LinearInterpolate(A=@n2.Normal, B=@n65, Alpha=@n66), Alpha=@n50)
  ) {attributeSetTypes [
    #2
    #3
    #4
    #5
    #6
    #1
   ]}
  FunctionOutput(None=StaticSwitchParameter(True=@n67, False=@n1) {group=Customization_Wrap, parameterName=Wrap_GoldenSkeleton})
  n68 = ScalarParameter() {defaultValue=20, group=Wrap_GoldenSkeleton, parameterName=GoldSkel_Tri_Contrast, sortPriority=-1}
  n69 = MaterialFunctionCall(In=Abs(None=ComponentMask(None=@n13) {b=true}), Contrast=@n68) {materialFunction=&CheapContrast}
  n70 = MaterialFunctionCall(In=Abs(None=ComponentMask(None=@n13) {g=true}), Contrast=@n68) {materialFunction=&CheapContrast}
  n71 = MaterialFunctionCall(In=Abs(None=ComponentMask(None=@n13) {r=true}), Contrast=@n68) {materialFunction=&CheapContrast}
  n72 = Add(A=Saturate(None=Subtract(A=@n69.Result, B=@n70.Result)), B=@n71.Result)
  n73 = Subtract(A=@n69.Result, B=@n71.Result)
  n74 = Saturate(None=Add(A=Saturate(None=@n73), B=@n70.Result))
  n75 = Subtract(A=@n70.Result, B=@n69.Result)
  n76 = Saturate(None=Add(A=Saturate(None=@n75), B=@n71.Result))
  n77 = MaterialFunctionCall(Color=@n5) {materialFunction=&SplitComponents}
  LinearInterpolate(A=LinearInterpolate(A=Saturate(None=@n72), B=LinearInterpolate(A=@n74, B=@n76, Alpha=@n77.G), Alpha=@n77.R), B=@n76, Alpha=@n77.B)
  MaterialFunctionCall(In=@n56, Contrast=ScalarParameter() {defaultValue=0.5, group=Wrap_GoldenSkeleton, parameterName=GoldSkel_GoldCruv_Contrast, sortPriority=0}) {materialFunction=&CheapContrast}
  Multiply(A=Add(A=@n55) {constB=-0.5}) {constB=2}
 ]}
