MaterialFunction /Game/Weapons/WeaponSkins/Wraps/MaterialFunctions/MF_Wrap_HeadC
graph {nodes [
  n1 = Constant()
  n2 = MaterialFunctionCall(
   R=FunctionInput(Preview=@n1) {inputName=isVehicle, inputType=FunctionInput_Scalar, sortPriority=5}
   G=FunctionInput(Preview=@n1) {bUsePreviewValueAsDefault=true, inputName=Side Project Vehicle, inputType=FunctionInput_Scalar}
   B=FunctionInput(Preview=@n1) {bUsePreviewValueAsDefault=true, inputName=isMech, inputType=FunctionInput_Scalar, sortPriority=4}
   A=NamedRerouteDeclaration(None=FunctionInput() {bUsePreviewValueAsDefault=true, inputName=isCharacter, inputType=FunctionInput_Scalar, sortPriority=100}) {name=IsCharacter, nodeColor=<rgba 0.3245 1 0 1>}
  ) {materialFunction=&AppendMany}
  n3 = MaterialFunctionCall(Color=@n2.RGB) {materialFunction=&SplitComponents}
  n4 = FunctionInput(Preview=VertexInterpolator(VS=PreSkinnedNormal()).PS) {bUsePreviewValueAsDefault=true, inputName=Pre-Skinned Local Normal, sortPriority=10}
  n5 = VertexInterpolator(VS=PreSkinnedPosition())
  n6 = PreSkinnedLocalBounds()
  n7 = FunctionInput(Preview=Divide(A=Subtract(A=@n5.PS, B=@n6.Min), B=Subtract(A=@n6.Max, B=@n6.Max))) {bUsePreviewValueAsDefault=true, inputName=Calcated_Bounds, sortPriority=11}
  n8 = MaterialFunctionCall!(
   For_Safe_Zones
   Normal_Map
   Caculated_Bounds=@n7
   Pre_Skinned_Normals=@n4
   Is Vehicle=@n3.R
   Is VehicleSide=@n3.R
   Is Mech=@n3.R
   Vehicle_Rotation
   UV_Rotation Bypass
   Mech_Rotation
   Uvs_Offset
   Offset-after rotation
   Uvs_TileRate
  ) {materialFunction=&/MF_Projection_Side_Normal}
  n9 = Divide(
   A=Subtract(A=Time(), B=ScalarParameter() {defaultValue=-9999999, group=Wrap_HeadCase, parameterName=HeadC_Game_TimeBPB, sortPriority=100})
   B=ScalarParameter() {defaultValue=1, group=Wrap_HeadCase, parameterName=HeadC_Duration_B, sortPriority=100}
  )
  n10 = Add(A=ComponentMask(None=@n8.UVs No Offsets/Tiling) {r=true}, B=LinearInterpolate(Alpha=Saturate(None=@n9)) {constA=-1})
  n11 = Multiply(A=@n10, B=OneMinus(None=@n10))
  n12 = LinearInterpolate(
   B=ScalarParameter() {defaultValue=2, group=Wrap_HeadCase, parameterName=HeadC_Kill_Scale_Effect, sortPriority=100}
   Alpha=Saturate(None=Multiply(A=@n11) {constB=4})
  ) {constA=1, constB=0}
  Add(A=@n12) {constB=-0.5}
  n13 = VectorParameter() {group=Wrap_HeadCase, parameterName=HeadC_Background_A_Speed_A, defaultValue=<rgba 0.05 -0.015 0 1>}
  n14 = Add(A=@n8.UVs No Offsets/Tiling, B=Multiply(A=ComponentMask(None=@n13.RGB) {g=true, r=true}, B=Time())) {constB=0}
  n15 = ScalarParameter() {defaultValue=5, group=Wrap_HeadCase, parameterName=HeadC_Background_A_TileRate_Vehicle}
  n16 = ScalarParameter() {defaultValue=1, group=Wrap_HeadCase, parameterName=HeadC_Background_A_TileRate_VehicleSide}
  n17 = MaterialFunctionCall(Color=@n2.RGB) {materialFunction=&SplitComponents}
  n18 = LinearInterpolate(
   A=LinearInterpolate(
    A=ScalarParameter() {defaultValue=5, group=Wrap_HeadCase, parameterName=HeadC_Background_A_TileRate_Weapon}
    B=LinearInterpolate(A=@n15, B=@n16, Alpha=@n17.G)
    Alpha=@n17.R
   )
   B=ScalarParameter() {defaultValue=1, group=Wrap_HeadCase, parameterName=HeadC_Background_A_TileRate_Mech}
   Alpha=@n17.B
  )
  n19 = ComponentMask(None=@n14) {r=true}
  n20 = Divide(A=Floor(None=Multiply(A=@n19, B=@n18) {constB=5}), B=@n18) {constB=5}
  n21 = ScalarParameter() {defaultValue=4, group=Wrap_HeadCase, parameterName=HeadC_Background_A_Offset_Y}
  n22 = Add(A=Multiply(A=@n14, B=@n18), B=AppendVector(A=Constant(), B=Multiply(A=@n20, B=@n21))) {constB=0}
  n23 = VectorParameter() {group=Wrap_HeadCase, parameterName=HeadC_Background_A_Random_seed, defaultValue=<rgba 4 3 0 0>}
  n24 = MaterialFunctionCall!(2 Vector=Add(A=Floor(None=@n22), B=ComponentMask(None=@n23.RGB) {g=true, r=true}), RandomSeed) {materialFunction=&/MF_Random_2}
  n25 = Multiply(
   A=Time()
   B=LinearInterpolate(
    A=ScalarParameter() {defaultValue=-0.25, group=Wrap_HeadCase, parameterName=HeadC_Background_A_Rotation_Speed_Range_A}
    B=ScalarParameter() {defaultValue=0.25, group=Wrap_HeadCase, parameterName=HeadC_Background_A_Rotation_Speed_Range_B}
    Alpha=@n24.utput
   )
  )
  Add(A=@n25, B=@n24.utput)
  n26 = FunctionInput() {bUsePreviewValueAsDefault=true, inputName=Material, inputType=FunctionInput_MaterialAttributes, sortPriority=-20}
  n27 = GetMaterialAttributes(None=@n26) {attributeGetTypes [
    #1
    #2
    #3
    #4
    #5
    #6
   ]}
  n28 = VectorParameter() {group=Wrap_HeadCase, parameterName=HeadC_BaseColor, sortPriority=0, defaultValue=<rgba 0.003 0.003 0.003 1>}
  n29 = FunctionInput() {bUsePreviewValueAsDefault=true, inputName="Customization Mask (_CM B)", inputType=FunctionInput_Scalar, sortPriority=24}
  n30 = Saturate(None=Subtract(A=NamedRerouteUsage(), B=NamedRerouteUsage()))
  n31 = Saturate(None=Add(
    A=LinearInterpolate(A=@n29, Alpha=@n30) {constB=0}
    B=FunctionInput() {bUsePreviewValueAsDefault=true, inputName="Customization Mask (_CM B)B", inputType=FunctionInput_Scalar, sortPriority=24}
   ) {constB=0})
  n32 = VectorParameter() {group=Wrap_HeadCase, parameterName=HeadC_Background_A_Color, defaultValue=<rgba 0 0.3906 0.2622 1>}
  n33 = VectorParameter() {group=Wrap_HeadCase, parameterName=HeadC_Background_B_Color, defaultValue=<rgba 1 1 1 1>}
  n34 = ScalarParameter() {defaultValue=1, group=Wrap_HeadCase, parameterName=HeadC_Background_C_TileRate_Vehicle}
  n35 = ScalarParameter() {defaultValue=1, group=Wrap_HeadCase, parameterName=HeadC_Background_C_TileRate_VehicleSide}
  n36 = MaterialFunctionCall(Color=@n2.RGB) {materialFunction=&SplitComponents}
  n37 = LinearInterpolate(
   A=LinearInterpolate(
    A=ScalarParameter() {defaultValue=2, group=Wrap_HeadCase, parameterName=HeadC_Background_C_TileRate_Weapon}
    B=LinearInterpolate(A=@n34, B=@n35, Alpha=@n36.G)
    Alpha=@n36.R
   )
   B=ScalarParameter() {defaultValue=0.5, group=Wrap_HeadCase, parameterName=HeadC_Background_C_TileRate_Mech}
   Alpha=@n36.B
  )
  n38 = ComponentMask(None=VectorParameter() {group=Wrap_HeadCase, parameterName=HeadC_Background_C_Speed_A}.RGB) {g=true, r=true}
  n39 = Time()
  n40 = TextureObjectParameter() {group=Wrap_HeadCase, parameterName=HeadC_Background__Tex, sortPriority=10, texture=&T_BlacksRGB}
  n41 = TextureSample(UVs=Add(A=Multiply(A=@n8.UVs No Offsets/Tiling, B=@n37), B=Multiply(A=@n38, B=@n39)), Tex=@n40) {samplerType=SAMPLERTYPE_LinearColor}
  n42 = Multiply(A=LinearInterpolate(A=@n32.RGB, B=@n33.RGB, Alpha=@n41.R), B=@n41.R)
  n43 = Multiply(A=Time(), B=ScalarParameter() {defaultValue=0.1, parameterName=HeadC_Background_Fade_Speed_A})
  n44 = Add(A=@n41.A, B=@n43)
  n45 = Saturate(None=LinearInterpolate(Alpha=Sine(None=@n44)) {constA=0.1, constB=1.25})
  n46 = ScalarParameter() {defaultValue=1, group=Wrap_HeadCase, parameterName=HeadC_Background_B_TileRate_Vehicle}
  n47 = ScalarParameter() {defaultValue=1, group=Wrap_HeadCase, parameterName=HeadC_Background_B_TileRate_VehicleSide}
  n48 = MaterialFunctionCall(Color=@n2.RGB) {materialFunction=&SplitComponents}
  n49 = LinearInterpolate(
   A=LinearInterpolate(
    A=ScalarParameter() {defaultValue=0.5, group=Wrap_HeadCase, parameterName=HeadC_Background_B_TileRate_Weapon}
    B=LinearInterpolate(A=@n46, B=@n47, Alpha=@n48.G)
    Alpha=@n48.R
   )
   B=ScalarParameter() {defaultValue=0.5, group=Wrap_HeadCase, parameterName=HeadC_Background_B_TileRate_Mech}
   Alpha=@n48.B
  )
  n50 = ComponentMask(None=VectorParameter() {group=Wrap_HeadCase, parameterName=HeadC_Background_B_Speed_A, defaultValue=<rgba -0.05 0.015 0 1>}.RGB) {g=true, r=true}
  n51 = Time()
  n52 = TextureSample(UVs=Add(A=Multiply(A=@n8.UVs No Offsets/Tiling, B=@n49), B=Multiply(A=@n50, B=@n51)), Tex=@n40) {samplerType=SAMPLERTYPE_LinearColor}
  n53 = LinearInterpolate(A=@n32.RGB, B=@n33.RGB, Alpha=@n52.R)
  n54 = Add(A=@n52.A, B=@n43)
  n55 = Saturate(None=LinearInterpolate(Alpha=Sine(None=@n54)) {constA=0.4, constB=1.25})
  n56 = Multiply(
   A=LinearInterpolate(A=Multiply(A=@n42, B=@n45), B=Multiply(A=@n53, B=@n52.R), Alpha=Multiply(A=@n52.G, B=@n55))
   B=ScalarParameter() {defaultValue=0.1, group=Wrap_HeadCase, parameterName=HeadC_Background_A_Brightness}
  )
  n57 = ScalarParameter() {group=Wrap_HeadCase, parameterName=HeadC_HeadG}
  n58 = Time()
  n59 = ScalarParameter() {defaultValue=-9999999, group=Wrap_HeadCase, parameterName=HeadC_Game_TimeBPA, sortPriority=100}
  n60 = Saturate(None=Divide(A=Subtract(A=@n58, B=@n59), B=ScalarParameter() {defaultValue=0.25, group=Wrap_HeadCase, parameterName=HeadC_Duration, sortPriority=100}))
  n61 = ScalarParameter() {group=Wrap_HeadCase, parameterName=HeadC_HeadF}
  n62 = LinearInterpolate(B=Multiply(A=Max(A=@n57, B=@n60), B=@n12), Alpha=@n61)
  n63 = MaterialFunctionCall(Color=@n2.RGB) {materialFunction=&SplitComponents}
  n64 = MaterialFunctionCall(Float4=NamedRerouteUsage()) {materialFunction=&BreakOutFloat4Components}
  n65 = LinearInterpolate(
   A=ScalarParameter() {group=Wrap_HeadCase, parameterName=HeadC_Head_Rotation_Sine_Amount_over}
   B=LinearInterpolate(
    A=ScalarParameter() {group=Wrap_HeadCase, parameterName=HeadC_Head_Rotation_Sine_Amount_over_Mech}
    B=ScalarParameter() {group=Wrap_HeadCase, parameterName=HeadC_Head_Rotation_Sine_Amount_over_Character}
    Alpha=@n64.A
   )
   Alpha=@n64.B
  )
  n66 = Multiply(A=Time(), B=ScalarParameter() {defaultValue=0.5, group=Wrap_HeadCase, parameterName=HeadC_Head_Rotation_Sine_Speed})
  n67 = Add(B=@n66) {constA=0.13, constB=0.03}
  n68 = ScalarParameter() {defaultValue=0.05, group=Wrap_HeadCase, parameterName=HeadC_Head_Rotation_Sine_Amount}
  n69 = Add(A=@n65, B=Multiply(A=Sine(None=@n67), B=@n68))
  n70 = VectorParameter() {group=Wrap_HeadCase, parameterName=HeadC_HeadF_Location_Weapon, defaultValue=<rgba 0 0 0 0.5>}
  n71 = VectorParameter() {group=Wrap_HeadCase, parameterName=HeadC_HeadF_Location_Vehicle, defaultValue=<rgba 0 0 0 5>}
  n72 = AppendVector(A=@n71.RGB, B=@n71.RGB)
  n73 = VectorParameter() {group=Wrap_HeadCase, parameterName=HeadC_HeadF_Location_VehicleSide, defaultValue=<rgba 0 0 0 5>}
  n74 = AppendVector(A=@n73.RGB, B=@n73.RGB)
  n75 = MaterialFunctionCall(Float4=NamedRerouteUsage()) {materialFunction=&BreakOutFloat4Components}
  n76 = VectorParameter() {group=Wrap_HeadCase, parameterName=HeadC_HeadF_Location_Mech, defaultValue=<rgba 0 0 0 0.5>}
  n77 = VectorParameter() {group=Wrap_HeadCase, parameterName=HeadC_HeadF_Location_Character, defaultValue=<rgba 0 0 0 0.5>}
  n78 = LinearInterpolate(
   A=LinearInterpolate(A=AppendVector(A=@n70.RGB, B=@n70.RGB), B=LinearInterpolate(A=@n72, B=@n74, Alpha=@n75.G), Alpha=@n75.R)
   B=LinearInterpolate(A=AppendVector(A=@n76.RGB, B=@n76.RGB), B=AppendVector(A=@n77.RGB, B=@n77.RGB), Alpha=@n75.A)
   Alpha=@n75.B
  )
  n79 = Multiply(A=Time(), B=ScalarParameter() {defaultValue=0.75, group=Wrap_HeadCase, parameterName=HeadC_Head_Scale_Sine_Speed})
  n80 = ScalarParameter() {defaultValue=0.04, group=Wrap_HeadCase, parameterName=HeadC_Head_Scale_Sine_Amount}
  n81 = Multiply(A=Sine(None=Add(B=@n79) {constA=0.33}), B=@n80)
  n82 = Multiply(A=ComponentMask(None=@n78) {a=true}, B=Constant3Vector() {constant=<rgba 1 0.75 0 0>})
  n83 = MaterialFunctionCall(
   For_Safe_Zones=StaticBool() {value=true}
   Caculated_Bounds=@n7
   Pre_Skinned_Normals=@n4
   Is Vehicle=@n63.R
   Is VehicleSide=@n63.R
   Is Mech=@n63.R
   UV_Rotation Bypass=@n69
   Uvs_Offset=ComponentMask(None=@n78) {b=true, g=true, r=true}
   Uvs_TileRate=Multiply(A=Add(A=@n81, B=@n82), B=@n62)
  ) {materialFunction=&MF_Projection_Side_Normal}
  n84 = Multiply(A=Divide() {constB=3}, B=Add(A=Constant2Vector() {g=1}, B=@n83.UVs))
  n85 = TextureObjectParameter() {group=Wrap_HeadCase, parameterName=HeadC_Mask__Tex, sortPriority=10, texture=&T_BlacksRGB}
  n86 = TextureObjectParameter() {group=Wrap_HeadCase, parameterName=HeadC_Clamp_Mask__Tex, sortPriority=10, texture=&T_BlacksRGB}
  n87 = Multiply(
   A=@n62
   B=Multiply(A=TextureSample(UVs=@n84, Tex=@n85).RGBA, B=TextureSample(UVs=@n83.UVs, Tex=@n86) {samplerType=SAMPLERTYPE_LinearGrayscale}.R)
  )
  n88 = ComponentMask(None=@n87) {b=true, g=true, r=true}
  n89 = ComponentMask(None=@n87) {a=true}
  n90 = ScalarParameter() {group=Wrap_HeadCase, parameterName=HeadC_HeadE}
  n91 = LinearInterpolate(B=Multiply(A=Max(A=@n60, B=@n61), B=@n12), Alpha=@n90)
  n92 = MaterialFunctionCall(Color=@n2.RGB) {materialFunction=&SplitComponents}
  n93 = Add(B=@n66) {constA=0.33, constB=0.03}
  n94 = Add(A=@n65, B=Multiply(A=Sine(None=@n93), B=@n68))
  n95 = VectorParameter() {group=Wrap_HeadCase, parameterName=HeadC_HeadE_Location_Weapon, defaultValue=<rgba 0 0 0 0.5>}
  n96 = VectorParameter() {group=Wrap_HeadCase, parameterName=HeadC_HeadE_Location_Vehicle, defaultValue=<rgba 0 0 0 5>}
  n97 = AppendVector(A=@n96.RGB, B=@n96.RGB)
  n98 = VectorParameter() {group=Wrap_HeadCase, parameterName=HeadC_HeadE_Location_VehicleSide, defaultValue=<rgba 0 0 0 5>}
  n99 = AppendVector(A=@n98.RGB, B=@n98.RGB)
  n100 = MaterialFunctionCall(Float4=NamedRerouteUsage()) {materialFunction=&BreakOutFloat4Components}
  n101 = VectorParameter() {group=Wrap_HeadCase, parameterName=HeadC_HeadE_Location_Mech, defaultValue=<rgba 0 0 0 0.5>}
  n102 = VectorParameter() {group=Wrap_HeadCase, parameterName=HeadC_HeadE_Location_Character, defaultValue=<rgba 0 0 0 0.5>}
  n103 = LinearInterpolate(
   A=LinearInterpolate(A=AppendVector(A=@n95.RGB, B=@n95.RGB), B=LinearInterpolate(A=@n97, B=@n99, Alpha=@n100.G), Alpha=@n100.R)
   B=LinearInterpolate(A=AppendVector(A=@n101.RGB, B=@n101.RGB), B=AppendVector(A=@n102.RGB, B=@n102.RGB), Alpha=@n100.A)
   Alpha=@n100.B
  )
  n104 = Multiply(A=Sine(None=Add(B=@n79) {constA=0.13}), B=@n80)
  n105 = Multiply(A=ComponentMask(None=@n103) {a=true}, B=Constant3Vector() {constant=<rgba 0.75 1 0 0>})
  n106 = MaterialFunctionCall(
   For_Safe_Zones=StaticBool() {value=true}
   Caculated_Bounds=@n7
   Pre_Skinned_Normals=@n4
   Is Vehicle=@n92.R
   Is VehicleSide=@n92.R
   Is Mech=@n92.R
   UV_Rotation Bypass=@n94
   Uvs_Offset=ComponentMask(None=@n103) {b=true, g=true, r=true}
   Uvs_TileRate=Multiply(A=Add(A=@n104, B=@n105), B=@n91)
  ) {materialFunction=&MF_Projection_Side_Normal}
  n107 = Multiply(A=Divide() {constB=3}, B=Add(A=Constant2Vector() {g=1}, B=@n106.UVs))
  n108 = Multiply(
   A=@n91
   B=Multiply(A=TextureSample(UVs=@n107, Tex=@n85).RGBA, B=TextureSample(UVs=@n106.UVs, Tex=@n86) {samplerType=SAMPLERTYPE_LinearGrayscale}.R)
  )
  n109 = ScalarParameter() {group=Wrap_HeadCase, parameterName=HeadC_HeadD}
  n110 = LinearInterpolate(B=Multiply(A=Max(A=@n90, B=@n60), B=@n12), Alpha=@n109)
  n111 = MaterialFunctionCall(Color=@n2.RGB) {materialFunction=&SplitComponents}
  n112 = Add(A=@n65, B=Multiply(A=Sine(None=@n66), B=@n68))
  n113 = VectorParameter() {group=Wrap_HeadCase, parameterName=HeadC_HeadD_Location_Weapon, defaultValue=<rgba 0 0 0 0.5>}
  n114 = VectorParameter() {group=Wrap_HeadCase, parameterName=HeadC_HeadD_Location_Vehicle, defaultValue=<rgba 0 0 0 5>}
  n115 = AppendVector(A=@n114.RGB, B=@n114.RGB)
  n116 = VectorParameter() {group=Wrap_HeadCase, parameterName=HeadC_HeadD_Location_VehicleSide, defaultValue=<rgba 0 0 0 5>}
  n117 = AppendVector(A=@n116.RGB, B=@n116.RGB)
  n118 = MaterialFunctionCall(Float4=NamedRerouteUsage()) {materialFunction=&BreakOutFloat4Components}
  n119 = VectorParameter() {group=Wrap_HeadCase, parameterName=HeadC_HeadD_Location_Mech, defaultValue=<rgba 0 0 0 0.5>}
  n120 = VectorParameter() {group=Wrap_HeadCase, parameterName=HeadC_HeadD_Location_Character, defaultValue=<rgba 0 0 0 0.5>}
  n121 = LinearInterpolate(
   A=LinearInterpolate(A=AppendVector(A=@n113.RGB, B=@n113.RGB), B=LinearInterpolate(A=@n115, B=@n117, Alpha=@n118.G), Alpha=@n118.R)
   B=LinearInterpolate(A=AppendVector(A=@n119.RGB, B=@n119.RGB), B=AppendVector(A=@n120.RGB, B=@n120.RGB), Alpha=@n118.A)
   Alpha=@n118.B
  )
  n122 = Multiply(A=Sine(None=@n79), B=@n80)
  n123 = ComponentMask(None=@n121) {a=true}
  n124 = MaterialFunctionCall(
   For_Safe_Zones=StaticBool() {value=true}
   Caculated_Bounds=@n7
   Is Vehicle=@n111.R
   Is VehicleSide=@n111.R
   Is Mech=@n111.R
   UV_Rotation Bypass=@n112
   Uvs_Offset=ComponentMask(None=@n121) {b=true, g=true, r=true}
   Uvs_TileRate=Multiply(A=Add(A=@n122, B=@n123), B=@n110)
  ) {materialFunction=&MF_Projection_Side_Normal}
  n125 = Multiply(A=Divide() {constB=3}, B=Add(A=Constant2Vector() {g=1}, B=@n124.UVs))
  n126 = Multiply(
   A=@n110
   B=Multiply(A=TextureSample(UVs=@n125, Tex=@n85).RGBA, B=TextureSample(UVs=@n124.UVs, Tex=@n86) {samplerType=SAMPLERTYPE_LinearGrayscale}.R)
  )
  n127 = LinearInterpolate(
   A=LinearInterpolate(
    A=LinearInterpolate(A=@n56, B=@n88, Alpha=@n89)
    B=ComponentMask(None=@n108) {b=true, g=true, r=true}
    Alpha=ComponentMask(None=@n108) {a=true}
   )
   B=ComponentMask(None=@n126) {b=true, g=true, r=true}
   Alpha=ComponentMask(None=@n126) {a=true}
  )
  n128 = ScalarParameter() {group=Wrap_HeadCase, parameterName=HeadC_HeadC}
  n129 = LinearInterpolate(B=Multiply(A=Max(A=@n109, B=@n60), B=@n12), Alpha=@n128)
  n130 = MaterialFunctionCall(Color=@n2.RGB) {materialFunction=&SplitComponents}
  n131 = VectorParameter() {group=Wrap_HeadCase, parameterName=HeadC_HeadC_Location_Weapon, defaultValue=<rgba 0 0 0 0.5>}
  n132 = VectorParameter() {group=Wrap_HeadCase, parameterName=HeadC_HeadC_Location_Vehicle, defaultValue=<rgba 0 0 0 5>}
  n133 = AppendVector(A=@n132.RGB, B=@n132.RGB)
  n134 = VectorParameter() {group=Wrap_HeadCase, parameterName=HeadC_HeadC_Location_VehicleSide, defaultValue=<rgba 0 0 0 5>}
  n135 = AppendVector(A=@n134.RGB, B=@n134.RGB)
  n136 = MaterialFunctionCall(Float4=NamedRerouteUsage()) {materialFunction=&BreakOutFloat4Components}
  n137 = VectorParameter() {group=Wrap_HeadCase, parameterName=HeadC_HeadC_Location_Mech, defaultValue=<rgba 0 0 0 0.5>}
  n138 = VectorParameter() {group=Wrap_HeadCase, parameterName=HeadC_HeadC_Location_Character, defaultValue=<rgba 0 0 0 0.5>}
  n139 = LinearInterpolate(
   A=LinearInterpolate(A=AppendVector(A=@n131.RGB, B=@n131.RGB), B=LinearInterpolate(A=@n133, B=@n135, Alpha=@n136.G), Alpha=@n136.R)
   B=LinearInterpolate(A=AppendVector(A=@n137.RGB, B=@n137.RGB), B=AppendVector(A=@n138.RGB, B=@n138.RGB), Alpha=@n136.A)
   Alpha=@n136.B
  )
  n140 = Multiply(A=ComponentMask(None=@n139) {a=true}, B=Constant3Vector() {constant=<rgba 1 0.75 0 0>})
  n141 = MaterialFunctionCall(
   For_Safe_Zones=StaticBool() {value=true}
   Caculated_Bounds=@n7
   Pre_Skinned_Normals=@n4
   Is Vehicle=@n130.R
   Is VehicleSide=@n130.R
   Is Mech=@n130.R
   UV_Rotation Bypass=@n94
   Uvs_Offset=ComponentMask(None=@n139) {b=true, g=true, r=true}
   Uvs_TileRate=Multiply(A=Add(A=@n81, B=@n140), B=@n129)
  ) {materialFunction=&MF_Projection_Side_Normal}
  n142 = Multiply(A=Divide() {constB=3}, B=Add(A=Constant2Vector() {r=3}, B=@n141.UVs))
  n143 = Multiply(
   A=@n129
   B=Multiply(A=TextureSample(UVs=@n142, Tex=@n85).RGBA, B=TextureSample(UVs=@n141.UVs, Tex=@n86) {samplerType=SAMPLERTYPE_LinearGrayscale}.R)
  )
  n144 = ComponentMask(None=@n143) {b=true, g=true, r=true}
  n145 = ComponentMask(None=@n143) {a=true}
  n146 = ScalarParameter() {group=Wrap_HeadCase, parameterName=HeadC_HeadB}
  n147 = LinearInterpolate(B=Multiply(A=Max(A=@n128, B=@n60), B=@n12), Alpha=@n146)
  n148 = MaterialFunctionCall(Color=@n2.RGB) {materialFunction=&SplitComponents}
  n149 = VectorParameter() {group=Wrap_HeadCase, parameterName=HeadC_HeadB_Location_Weapon, defaultValue=<rgba 0 0 0 0.5>}
  n150 = VectorParameter() {group=Wrap_HeadCase, parameterName=HeadC_HeadB_Location_Vehicle, defaultValue=<rgba 0 0 0 5>}
  n151 = AppendVector(A=@n150.RGB, B=@n150.RGB)
  n152 = VectorParameter() {group=Wrap_HeadCase, parameterName=HeadC_HeadB_Location_VehicleSide, defaultValue=<rgba 0 0 0 5>}
  n153 = AppendVector(A=@n152.RGB, B=@n152.RGB)
  n154 = MaterialFunctionCall(Float4=NamedRerouteUsage()) {materialFunction=&BreakOutFloat4Components}
  n155 = VectorParameter() {group=Wrap_HeadCase, parameterName=HeadC_HeadB_Location_Mech, defaultValue=<rgba 0 0 0 0.5>}
  n156 = VectorParameter() {group=Wrap_HeadCase, parameterName=HeadC_HeadB_Location_Character, defaultValue=<rgba 0 0 0 0.5>}
  n157 = LinearInterpolate(
   A=LinearInterpolate(A=AppendVector(A=@n149.RGB, B=@n149.RGB), B=LinearInterpolate(A=@n151, B=@n153, Alpha=@n154.G), Alpha=@n154.R)
   B=LinearInterpolate(A=AppendVector(A=@n155.RGB, B=@n155.RGB), B=AppendVector(A=@n156.RGB, B=@n156.RGB), Alpha=@n154.A)
   Alpha=@n154.B
  )
  n158 = Multiply(A=ComponentMask(None=@n157) {a=true}, B=Constant3Vector() {constant=<rgba 0.75 1 0 0>})
  n159 = MaterialFunctionCall(
   For_Safe_Zones=StaticBool() {value=true}
   Caculated_Bounds=@n7
   Pre_Skinned_Normals=@n4
   Is Vehicle=@n148.R
   Is VehicleSide=@n148.R
   Is Mech=@n148.R
   UV_Rotation Bypass=@n69
   Uvs_Offset=ComponentMask(None=@n157) {b=true, g=true, r=true}
   Uvs_TileRate=Multiply(A=Add(A=@n104, B=@n158), B=@n147)
  ) {materialFunction=&MF_Projection_Side_Normal}
  n160 = Multiply(A=Divide() {constB=3}, B=Add(A=Constant2Vector() {r=2}, B=@n159.UVs))
  n161 = Multiply(
   A=@n147
   B=Multiply(A=TextureSample(UVs=@n160, Tex=@n85).RGBA, B=TextureSample(UVs=@n159.UVs, Tex=@n86) {samplerType=SAMPLERTYPE_LinearGrayscale}.R)
  )
  n162 = MaterialFunctionCall(Color=@n2.RGB) {materialFunction=&SplitComponents}
  n163 = VectorParameter() {group=Wrap_HeadCase, parameterName=HeadC_HeadA_Location_Weapon, defaultValue=<rgba 0.1 0 0 0.5>}
  n164 = VectorParameter() {group=Wrap_HeadCase, parameterName=HeadC_HeadA_Location_Vehicle, defaultValue=<rgba 0 0 0 5>}
  n165 = AppendVector(A=@n164.RGB, B=@n164.RGB)
  n166 = VectorParameter() {group=Wrap_HeadCase, parameterName=HeadC_HeadA_Location_VehicleSide, defaultValue=<rgba 0 0 0 5>}
  n167 = AppendVector(A=@n166.RGB, B=@n166.RGB)
  n168 = MaterialFunctionCall(Float4=NamedRerouteUsage()) {materialFunction=&BreakOutFloat4Components}
  n169 = VectorParameter() {group=Wrap_HeadCase, parameterName=HeadC_HeadA_Location_Mech, defaultValue=<rgba 0 0 0 0.5>}
  n170 = VectorParameter() {group=Wrap_HeadCase, parameterName=HeadC_HeadA_Location_Character, defaultValue=<rgba 0 0 0 0.5>}
  n171 = LinearInterpolate(
   A=LinearInterpolate(A=AppendVector(A=@n163.RGB, B=@n163.RGB), B=LinearInterpolate(A=@n165, B=@n167, Alpha=@n168.G), Alpha=@n168.R)
   B=LinearInterpolate(A=AppendVector(A=@n169.RGB, B=@n169.RGB), B=AppendVector(A=@n170.RGB, B=@n170.RGB), Alpha=@n168.A)
   Alpha=@n168.B
  )
  n172 = ComponentMask(None=@n171) {a=true}
  n173 = LinearInterpolate(B=Multiply(A=Max(A=@n146, B=@n60), B=@n12), Alpha=ScalarParameter() {group=Wrap_HeadCase, parameterName=HeadC_HeadA})
  n174 = MaterialFunctionCall(
   For_Safe_Zones=StaticBool() {value=true}
   Caculated_Bounds=@n7
   Pre_Skinned_Normals=@n4
   Is Vehicle=@n162.R
   Is VehicleSide=@n162.R
   Is Mech=@n162.R
   UV_Rotation Bypass=@n112
   Uvs_Offset=ComponentMask(None=@n171) {b=true, g=true, r=true}
   Uvs_TileRate=Multiply(A=Add(A=@n122, B=@n172), B=@n173)
  ) {materialFunction=&MF_Projection_Side_Normal}
  n175 = Multiply(A=Divide() {constB=3}, B=Add(A=Constant2Vector() {r=1}, B=@n174.UVs))
  n176 = Multiply(
   A=@n173
   B=Multiply(A=TextureSample(UVs=@n175, Tex=@n85).RGBA, B=TextureSample(UVs=@n174.UVs, Tex=@n86) {samplerType=SAMPLERTYPE_LinearGrayscale}.R)
  )
  n177 = LinearInterpolate(
   A=LinearInterpolate(
    A=LinearInterpolate(A=@n127, B=@n144, Alpha=@n145)
    B=ComponentMask(None=@n161) {b=true, g=true, r=true}
    Alpha=ComponentMask(None=@n161) {a=true}
   )
   B=ComponentMask(None=@n176) {b=true, g=true, r=true}
   Alpha=ComponentMask(None=@n176) {a=true}
  )
  n178 = MaterialFunctionCall(Color=Abs(None=@n4)) {materialFunction=&SplitComponents}
  n179 = MaterialFunctionCall(Color=@n2.RGB) {materialFunction=&SplitComponents}
  n180 = LinearInterpolate(A=@n178.B, B=@n178.B, Alpha=@n179.G)
  n181 = MaterialFunctionCall(
   In=LinearInterpolate(A=LinearInterpolate(A=@n178.R, B=@n180, Alpha=@n179.R), B=@n178.G, Alpha=@n179.B)
   Contrast=ScalarParameter() {defaultValue=1, group=Wrap_HeadCase, parameterName="HeadC_Triplaner Contrast ", sortPriority=-1}
  ) {materialFunction=&CheapContrast}
  n182 = VectorParameter() {group=Wrap_HeadCase, parameterName=HeadC_Edge_Color, defaultValue=<rgba 1 1 1 1>}
  n183 = FunctionInput() {bUsePreviewValueAsDefault=true, inputName=Curavture, inputType=FunctionInput_Scalar, sortPriority=24}
  n184 = Add(A=@n183) {constB=-0.5}
  n185 = MaterialFunctionCall(In=Saturate(None=Multiply(A=@n184) {constB=2}), Contrast=@n182.A) {materialFunction=&CheapContrast}
  n186 = TextureCoordinate() {uTiling=2, vTiling=2}
  n187 = ScalarParameter() {defaultValue=1, group=Wrap_HeadCase, parameterName=HeadC_FresA_Edge_Tiling}
  n188 = ScalarParameter() {group=Wrap_HeadCase, parameterName=HeadC_FresA_GradientSpeed_X}
  n189 = ScalarParameter() {defaultValue=0.1, group=Wrap_HeadCase, parameterName=HeadC_FresA_GradientSpeed_Y}
  n190 = TextureSampleParameter2D(UVs=Panner(Coordinate=Multiply(A=@n186, B=@n187), Speed=AppendVector(A=@n188, B=@n189))) {group=Wrap_HeadCase, parameterName=HeadC_FresA___Tex, texture=&T_BlacksRGB}
  n191 = Saturate(None=Fresnel(ExponentIn=ScalarParameter() {defaultValue=5, group=Wrap_HeadCase, parameterName=HeadC_FresA_Exponent}) {baseReflectFraction=0})
  n192 = Fresnel(ExponentIn=ScalarParameter() {defaultValue=5, group=Wrap_HeadCase, parameterName=HeadC_FresB_OutlineExponent1}, Normal=VertexNormalWS()) {baseReflectFraction=0}
  n193 = ScalarParameter() {defaultValue=5, group=Wrap_HeadCase, parameterName=HeadC_FresB_Outline_Brightness}
  n194 = Add(
   A=Multiply(A=Multiply(A=@n190.R, B=@n191), B=ScalarParameter() {defaultValue=5, group=Wrap_HeadCase, parameterName=HeadC_FresA_Brightness})
   B=Saturate(None=Multiply(A=@n192, B=@n193))
  )
  n195 = VectorParameter() {group=Wrap_HeadCase, parameterName=HeadC_Fres_ColorA, defaultValue=<rgba 0.1507 0.3271 0.3594 1>}
  n196 = VectorParameter() {group=Wrap_HeadCase, parameterName=HeadC_Fres_ColorB, defaultValue=<rgba 1 1 1 1>}
  n197 = Add(
   A=Add(A=Multiply(A=@n182.RGB, B=@n185.Result) {constB=2}, B=Multiply(A=@n181.Result, B=@n177))
   B=Multiply(A=@n194, B=LinearInterpolate(A=@n195.RGB, B=@n196.RGB, Alpha=@n194))
  )
  n198 = SetMaterialAttributes!(
   MaterialAttributes=@n27.MaterialAttributes
   BaseColor=LinearInterpolate(A=@n27.BaseColor, B=@n28.RGB, Alpha=@n31)
   Roughness=LinearInterpolate(A=@n27.Roughness, B=ScalarParameter() {defaultValue=0.5, group=Wrap_HeadCase, parameterName=HeadC_Roghness, sortPriority=0}, Alpha=@n31)
   Specular=LinearInterpolate(A=@n27.Specular, B=ScalarParameter() {defaultValue=0.5, group=Wrap_HeadCase, parameterName=HeadC_Specular, sortPriority=0}, Alpha=@n31)
   EmissiveColor=Add(A=@n27.EmissiveColor, B=Multiply(A=@n31, B=@n197))
   Metallic=LinearInterpolate(A=@n27.Metallic, B=ScalarParameter() {group=Wrap_HeadCase, parameterName=HeadC_Metallic, sortPriority=0}, Alpha=@n31)
   Normal=@n27.MaterialAttributes
  ) {attributeSetTypes [
    #1
    #2
    #3
    #4
    #5
    #6
   ]}
  BlendMaterialAttributes(A=@n27.MaterialAttributes, B=@n198, Alpha=@n31)
  Constant() {r=1}
  FunctionInput(Preview=Constant3Vector() {constant=<rgba 0 1 0 0>}) {bUsePreviewValueAsDefault=true, inputName=ForwardAxis}
  FunctionOutput(None=StaticSwitchParameter(True=@n198, False=@n26) {group=Customization_Wrap, parameterName=Wrap_HeadCase})
  n199 = Constant() {r=1}
  MaterialFunctionCall(Input=@n183, Input Low=Constant() {r=0.5}, Input High=@n199, Target Low=Constant(), Target High=@n199) {materialFunction=&RemapValueRange}
  Multiply(A=@n88, B=@n89)
  n200 = Constant() {r=3}
  n201 = ScalarParameter() {defaultValue=0.8, group=Wrap_HeadCase, parameterName=HeadC_Background_A_Scale}
  n202 = Time()
  n203 = Multiply(A=ScalarParameter() {defaultValue=1, group=Wrap_HeadCase, parameterName=HeadC_Background_A_Scale_Speed}, B=@n24.utput)
  n204 = Multiply(
   A=Sine(None=Multiply(A=@n202, B=@n203))
   B=ScalarParameter() {defaultValue=0.1, group=Wrap_HeadCase, parameterName=HeadC_Background_A_Scale_Sine_Amount}
  )
  n205 = MaterialFunctionCall(
   UVs=MaterialFunctionCall(UVs=Frac(None=@n22), Texture Scale=Add(A=@n201, B=@n204) {constA=0.8}) {materialFunction=&ScaleUVsByCenter}.UVs
   "Rotation Angle (0-1)"=@n25
  ) {materialFunction=&CustomRotator}
  n206 = MaterialFunctionCall(
   "Animation  Phase (0-1)"=LinearInterpolate(Alpha=@n24.utput) {constA=-3, constB=3}
   Number of Rows=@n200
   Number of Columns=@n200
   Texture=@n85
   UVs=@n205.Rotated Values
  ) {materialFunction=&FlipBook}
  n207 = MaterialFunctionCall!(UVs=@n8.UVs No Offsets/Tiling, Texture Size=ScalarParameter() {group=Wrap_HeadCase, parameterName=HeadC_Background_A_Miplevel _Size}) {materialFunction=/Engine/Functions/Engine_MaterialFunctions01/Texturing/ComputeMipLevel}
  n208 = Time()
  n209 = ScalarParameter() {defaultValue=0.05, group=Wrap_HeadCase, parameterName=HeadC_Background_A_Fade_Speed}
  n210 = Sine(None=Add(A=@n24.utput, B=Multiply(A=@n208, B=@n209)))
  Multiply(
   A=Multiply(
    A=MaterialFunctionCall(Color=@n206.Result) {materialFunction=&SplitComponents}.RGB
    B=TextureSample!(UVs=@n205.Rotated Values, Tex=@n86, Bias=@n207.MipLevel, Apply View MipBias) {mipValueMode=TMVM_MipBias, samplerType=SAMPLERTYPE_LinearGrayscale}.R
   )
   B=Saturate(None=LinearInterpolate(Alpha=@n210) {constA=0.4, constB=1.25})
  )
  n211 = ScalarParameter() {defaultValue=1, group=Wrap_HeadCase, parameterName=HeadC_Edge_Noise_effect}
  Multiply(A=Saturate(None=Add(A=@n190.R, B=@n211) {constB=0.5}), B=@n185.Result)
  Multiply(A=MaterialFunctionCall!(X range Y range, Gradient Height, ShowGrid?, Thickness, Color, "f(x)"=@n55) {materialFunction=/Engine/Functions/Engine_MaterialFunctions01/Debug/PlotFunctionOnGraph}."f(x) = y")
  Multiply(A=MaterialFunctionCall("f(x)"=Constant() {r=1}) {materialFunction=&PlotFunctionOnGraph}."f(x) = y")
  NamedRerouteDeclaration(None=@n2.RGBA) {name=IsState, nodeColor=<rgba 0.8271 1 0.005 1>}
  NamedRerouteDeclaration(None=FunctionInput() {bUsePreviewValueAsDefault=true, inputName=isCharAccent, inputType=FunctionInput_Scalar, sortPriority=78}) {name=isCharAccent, nodeColor=<rgba 0 0.9253 1 1>}
  OneMinus()
  Power(Base=@n192, Exp=ScalarParameter() {defaultValue=1, group=Wrap_HeadCase, parameterName=HeadC_FresB_OutlineExponent2})
  Saturate(None=Add(A=@n190.R) {constB=0.25})
  ScalarParameter() {defaultValue=5, group=Wrap_HeadCase, parameterName=HeadC_Background_A_TileRate}
  ScalarParameter() {defaultValue=0.5, group=Wrap_HeadCase, parameterName=HeadC_2nd_Specular, sortPriority=0}
  ScalarParameter() {defaultValue=0.5, group=Wrap_HeadCase, parameterName=HeadC_2nd_Roghness, sortPriority=0}
  SphereMask(A=@n205.Rotated Values, B=Constant() {r=0.5}, Radius=Constant() {r=0.5}, Hardness=Constant() {r=0.9})
  SphereMask(A=@n174.UVs, B=Constant() {r=0.5}, Hardness=Constant() {r=0.9}) {attenuationRadius=0.5, hardnessPercent=1}
  TextureSample!(UVs=Multiply(A=Divide() {constB=3}, B=Add(A=@n205.Rotated Values)), Tex=@n85, Apply View MipBias) {samplerType=SAMPLERTYPE_LinearGrayscale}
  VectorParameter() {group=Wrap_HeadCase, parameterName=HeadC_2nd_BaseColor, sortPriority=0, defaultValue=<rgba 0.003 0.003 0.003 1>}
 ]}
