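# MF_Wrap_HexColor is a weapon-skin wrap material function that blends an animated hex-color look onto an incoming material through primary and secondary masks, behind a static switch.
# Its back material adds emissive from a safe-zone virtual-plane projection and three height-offset hex layers per side, tinted by a scrolling curve-atlas gradient.
# A secondary material scrolls that gradient along a forward axis.
# IsState inputs select per-object-type axes, tiling and heights.
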MaterialFunction /Game/Weapons/WeaponSkins/Wraps/MaterialFunctions/MF_Wrap_HexColor
graph {nodes [
  n1 = NamedRerouteUsage()
  n2 = NamedRerouteUsage()
  n3 = MaterialFunctionCall(Float4=NamedRerouteUsage()) {materialFunction=&BreakOutFloat4Components}
  ComponentMask(None=MaterialFunctionCall(
    UVs=MaterialFunctionCall!(
     For_Safe_Zones
     Normal_Map
     Caculated_Bounds=@n1
     Pre_Skinned_Normals=@n2
     Is Vehicle=@n3.R
     Is VehicleSide=@n3.R
     Is Mech=@n3.R
     Vehicle_Rotation
     UV_Rotation Bypass
     Mech_Rotation
     Uvs_Offset
     Offset-after rotation
     Uvs_TileRate
    ) {materialFunction=&/MF_Projection_Side_Normal}.UVs No Offsets/Tiling
    Rotation Center=Constant() {r=0.5}
    "Rotation Angle (0-1)"=NamedRerouteUsage()
   ) {materialFunction=&CustomRotator}.Rotated Values) {r=true}
  Constant4Vector()
  n4 = TextureSample() {samplerType=SAMPLERTYPE_Masks, texture=/Game/Weapons/FORT_Rifles/Textures/T_Scar_CM}
  n5 = FunctionInput(Preview=@n4.R) {bUsePreviewValueAsDefault=true, inputName=Mask Primary, inputType=FunctionInput_Scalar, sortPriority=24, previewValue=<1 0 0 1>}
  n6 = BlendMaterialAttributes(A=NamedRerouteUsage(), B=NamedRerouteUsage(), Alpha=@n5) {vertexAttributeBlendType=UseA}
  n7 = NamedRerouteUsage()
  n8 = FunctionInput(Preview=@n4.B) {bUsePreviewValueAsDefault=true, inputName=Mask Secondary, inputType=FunctionInput_Scalar, sortPriority=24, previewValue=<1 0 0 1>}
  n9 = TextureSample() {samplerType=SAMPLERTYPE_Normal, texture=/Game/Weapons/FORT_Rifles/T_SCAR_N}
  n10 = Constant3Vector() {constant=<rgba 1 0 0 0>}
  n11 = GetMaterialAttributes(None=FunctionInput(Preview=SetMaterialAttributes!(MaterialAttributes, Normal=@n9.RGB, BaseColor=@n10) {attributeSetTypes [#1, #2]}) {inputName=Material, inputType=FunctionInput_MaterialAttributes}) {attributeGetTypes [
    #2
    #3
    #1
    #4
   ]}
  FunctionOutput(None=StaticSwitchParameter(
    True=BlendMaterialAttributes(A=@n6, B=@n7, Alpha=@n8) {vertexAttributeBlendType=UseA}
    False=NamedRerouteDeclaration(None=@n11.MaterialAttributes) {name=Root_Material, nodeColor=<rgba 0.7812 1 0 1>}
   ) {group=Customization_Wrap, parameterName=Wrap_HexColor}) {outputName=Result_1}
  n12 = VectorParameter() {group=Wrap_HexColor, parameterName=HexC_BC_Color_B, defaultValue=<rgba 0.0222 0.008 0.0578 1>}
  n13 = Add(A=OneMinus(None=MaterialFunctionCall() {materialFunction=/Engine/Functions/Engine_MaterialFunctions02/WorldPositionOffset/ObjectScale}.Scale Z)) {constB=-1}
  n14 = PreSkinnedLocalBounds()
  n15 = MaterialFunctionCall(Float3=@n14.Extents) {materialFunction=&BreakOutFloat3Components}
  n16 = FunctionInput(Preview=Constant4Vector() {constant=<rgba 0 0 0 1>}.RGBA) {bUsePreviewValueAsDefault=true, inputName=Object-SafeZone-Virtual-Pivot, inputType=FunctionInput_Vector4, sortPriority=105}
  n17 = VectorParameter() {group=Wrap_HexColor_2nd_Uvs, parameterName="HexC_VP_Offset_A=Scale"}
  n18 = Add(A=@n16, B=AppendVector(A=@n17.RGB, B=@n17.RGB))
  n19 = Multiply(A=Max(A=Max(A=@n15.R, B=@n15.R), B=@n15.B), B=ComponentMask(None=@n18) {a=true})
  n20 = FunctionInput(Preview=Constant()) {bUsePreviewValueAsDefault=true, inputName=Object-SafeZone-VP_Charcter_Preview_Switch, inputType=FunctionInput_Scalar, sortPriority=105}
  n21 = MaterialFunctionCall!(
   Camera _pos
   Axis
   Scale=NamedRerouteDeclaration(None=Multiply(A=@n13, B=@n19)) {name=Scale, nodeColor=<rgba 1 0.0234 0 1>}
   Scale_Multplyer
   Object-SafeZone-Virtual-Pivot=NamedRerouteDeclaration(None=@n18) {name=SafeZone_VP, nodeColor=<rgba 0.9775 0.0631 1 1>}
   Object-SafeZone-Virtual-CH-Switch=NamedRerouteDeclaration(None=@n20) {name=SafeZone_CH_Switch, nodeColor=<rgba 0.9775 0.0631 1 1>}
   SinglePlane Normal Axis=NamedRerouteUsage()
   SinglePlane X Axis=NamedRerouteUsage()
   PlaneCenter_distortion
  ) {materialFunction=&/MF_SafeZone_VirtualPlane}
  n22 = Add(A=@n21.Single_plane_UVs, B=NamedRerouteUsage())
  n23 = TextureSample(UVs=Multiply(A=@n22, B=NamedRerouteUsage()), Tex=NamedRerouteUsage())
  LinearInterpolate(A=VectorParameter() {group=Wrap_HexColor, parameterName=HexC_BC_Color_A, defaultValue=<rgba 0 0 0 1>}.RGB, B=@n12.RGB, Alpha=@n23.A)
  LinearInterpolate(B=VectorParameter() {group=Wrap_HexColor, parameterName=HexC_BC_Color_C, defaultValue=<rgba 0.0085 0.0032 0.0208 1>}.RGB)
  n24 = NamedRerouteUsage()
  n25 = MaterialFunctionCall!(
   Warp=NamedRerouteUsage()
   Height_Offset=Multiply(A=ComponentMask(None=@n24) {r=true}) {constB=-1}
   Constrant=NamedRerouteUsage()
   Scale=NamedRerouteUsage()
   SafeZone_Pivot=NamedRerouteUsage()
   SafeZone_Chr_Switch=NamedRerouteUsage()
   SinglePlane_Normal=NamedRerouteUsage()
   SinglePlane_X_Axis=NamedRerouteUsage()
   Tile_Rate_Master=NamedRerouteUsage()
   Tex_HexMsk=NamedRerouteUsage()
   Tex_Warp=NamedRerouteUsage()
   Tex_Noise=NamedRerouteUsage()
   Speed=NamedRerouteUsage()
   PlaneCenter_distortion
  ) {materialFunction=&/MF_Wrap_HexColor_Layers}
  n26 = Multiply(A=@n25.Hex_Msk, B=ScalarParameter() {defaultValue=0.05, group=Wrap_HexColor_Layers, parameterName=HexC_Opacity_A})
  Multiply(A=@n26, B=@n25.Result)
  n27 = ComponentMask(None=VectorParameter() {group=Wrap_HexColor_Layers, parameterName=HexC_Height_Tilerate_, defaultValue=<rgba 0.25 0.01 0 1>}.RGB) {g=true, r=true}
  n28 = TextureSample(UVs=Add(A=Multiply(A=@n21.Single_plane_UVs, B=@n27), B=NamedRerouteUsage()), Tex=NamedRerouteUsage()) {samplerSource=SSM_Wrap_WorldGroupSettings}
  Multiply(A=@n28.A)
  n29 = MaterialFunctionCall(
   Warp=NamedRerouteUsage()
   Height_Offset=ComponentMask(None=NamedRerouteUsage()) {r=true}
   Constrant=NamedRerouteUsage()
   Scale=NamedRerouteUsage()
   SafeZone_Pivot=NamedRerouteUsage()
   SafeZone_Chr_Switch=NamedRerouteUsage()
   SinglePlane_Normal=NamedRerouteUsage()
   SinglePlane_X_Axis=NamedRerouteUsage()
   Tile_Rate_Master=NamedRerouteUsage()
   Tex_HexMsk=NamedRerouteUsage()
   Tex_Warp=NamedRerouteUsage()
   Tex_Noise=NamedRerouteUsage()
   Speed=NamedRerouteUsage()
  ) {materialFunction=&MF_Wrap_HexColor_Layers}
  n30 = Multiply(A=@n29.Hex_Msk, B=ScalarParameter() {defaultValue=0.05, group=Wrap_HexColor_Layers, parameterName=HexC_Opacity_A})
  Multiply(A=@n30, B=@n29.Result)
  NamedRerouteDeclaration(None=TextureObjectParameter() {group=Wrap_HexColor, parameterName=HexC_Tex_Hex_Msk, texture=&T_BlacksRGB}) {name=Tex_Hex_Msk, nodeColor=<rgba 0.2656 0 1 1>}
  NamedRerouteDeclaration(None=TextureObjectParameter() {group=Wrap_HexColor, parameterName=HexC_Tex_Hex_Noise, texture=&T_BlacksRGB}) {name=Tex_Hex_Noise, nodeColor=<rgba 0.2656 0 1 1>}
  n31 = ComponentMask(None=NamedRerouteUsage()) {g=true}
  n32 = Multiply(
   A=MaterialFunctionCall(
    Warp=NamedRerouteUsage()
    Height_Offset=Multiply(A=@n31) {constB=-1}
    Constrant=NamedRerouteUsage()
    Scale=NamedRerouteUsage()
    SafeZone_Pivot=NamedRerouteUsage()
    SafeZone_Chr_Switch=NamedRerouteUsage()
    SinglePlane_Normal=NamedRerouteUsage()
    SinglePlane_X_Axis=NamedRerouteUsage()
    Tile_Rate_Master=NamedRerouteUsage()
    Tex_HexMsk=NamedRerouteUsage()
    Tex_Warp=NamedRerouteUsage()
    Tex_Noise=NamedRerouteUsage()
    Speed=NamedRerouteUsage()
   ) {materialFunction=&MF_Wrap_HexColor_Layers}.Hex_Msk
   B=ScalarParameter() {defaultValue=0.25, group=Wrap_HexColor_Layers, parameterName=HexC_Opacity_B}
  )
  n33 = NamedRerouteUsage()
  n34 = MaterialFunctionCall(
   Warp=NamedRerouteUsage()
   Height_Offset=Multiply(A=ComponentMask(None=@n33) {b=true}) {constB=-1}
   Constrant=NamedRerouteUsage()
   Scale=NamedRerouteUsage()
   SafeZone_Pivot=NamedRerouteUsage()
   SafeZone_Chr_Switch=NamedRerouteUsage()
   SinglePlane_Normal=NamedRerouteUsage()
   SinglePlane_X_Axis=NamedRerouteUsage()
   Tile_Rate_Master=NamedRerouteUsage()
   Tex_HexMsk=NamedRerouteUsage()
   Tex_Warp=NamedRerouteUsage()
   Tex_Noise=NamedRerouteUsage()
   Speed=NamedRerouteUsage()
  ) {materialFunction=&MF_Wrap_HexColor_Layers}
  n35 = ScalarParameter() {defaultValue=1, group=Wrap_HexColor_Layers, parameterName=HexC_Opacity_C}
  NamedRerouteDeclaration(None=LinearInterpolate(A=LinearInterpolate(A=@n26, Alpha=@n32), Alpha=Multiply(A=@n34.Hex_Msk, B=@n35))) {name=Side_Right, nodeColor=<rgba 0 0.0625 1 1>}
  n36 = NamedRerouteUsage()
  n37 = Multiply(
   A=MaterialFunctionCall(
    Warp=NamedRerouteUsage()
    Height_Offset=ComponentMask(None=@n36) {g=true}
    Constrant=NamedRerouteUsage()
    Scale=NamedRerouteUsage()
    SafeZone_Pivot=NamedRerouteUsage()
    SafeZone_Chr_Switch=NamedRerouteUsage()
    SinglePlane_Normal=NamedRerouteUsage()
    SinglePlane_X_Axis=NamedRerouteUsage()
    Tile_Rate_Master=NamedRerouteUsage()
    Tex_HexMsk=NamedRerouteUsage()
    Tex_Warp=NamedRerouteUsage()
    Tex_Noise=NamedRerouteUsage()
    Speed=NamedRerouteUsage()
   ) {materialFunction=&MF_Wrap_HexColor_Layers}.Hex_Msk
   B=ScalarParameter() {defaultValue=0.25, group=Wrap_HexColor_Layers, parameterName=HexC_Opacity_B}
  )
  n38 = MaterialFunctionCall(
   Warp=NamedRerouteUsage()
   Height_Offset=ComponentMask(None=NamedRerouteUsage()) {b=true}
   Constrant=NamedRerouteUsage()
   Scale=NamedRerouteUsage()
   SafeZone_Pivot=NamedRerouteUsage()
   SafeZone_Chr_Switch=NamedRerouteUsage()
   SinglePlane_Normal=NamedRerouteUsage()
   SinglePlane_X_Axis=NamedRerouteUsage()
   Tile_Rate_Master=NamedRerouteUsage()
   Tex_HexMsk=NamedRerouteUsage()
   Tex_Warp=NamedRerouteUsage()
   Tex_Noise=NamedRerouteUsage()
   Speed=NamedRerouteUsage()
  ) {materialFunction=&MF_Wrap_HexColor_Layers}
  n39 = ScalarParameter() {defaultValue=1, group=Wrap_HexColor_Layers, parameterName=HexC_Opacity_C}
  NamedRerouteDeclaration(None=LinearInterpolate(A=LinearInterpolate(A=@n30, Alpha=@n37), Alpha=Multiply(A=@n38.Hex_Msk, B=@n39))) {name=Side_Left, nodeColor=<rgba 0 0.0625 1 1>}
  n40 = VertexInterpolator(VS=PreSkinnedPosition())
  n41 = FunctionInput(Preview=@n40.PS) {bUsePreviewValueAsDefault=true, inputName=Pre-Skinned Local Postion, sortPriority=10}
  n42 = PreSkinnedLocalBounds()
  n43 = Subtract(A=@n42.Max, B=@n42.Max)
  NamedRerouteDeclaration(None=Divide(A=Subtract(A=@n41, B=@n42.Min), B=@n43)) {name=PreSkin_Position, nodeColor=<rgba 0.8281 1 0 1>}
  n44 = MaterialFunctionCall(Float3=Abs(None=NamedRerouteUsage())) {materialFunction=&BreakOutFloat3Components}
  n45 = Constant() {r=0.5}
  n46 = NamedRerouteUsage()
  n47 = ComponentMask(None=@n46) {desc=YX, g=true, r=true}
  n48 = If(A=@n44.R, B=@n45, "A > B"=ComponentMask(None=@n46) {b=true, desc=YZ, g=true}, "A < B"=ComponentMask(None=@n46) {b=true, desc=XZ, r=true})
  NamedRerouteDeclaration(None=Multiply(
    A=If(A=@n44.B, B=@n45, "A > B"=@n47, "A < B"=@n48)
    B=ScalarParameter() {defaultValue=1, group=Wrap_HexColor, parameterName=HexC_CheapTri_TileRate}
   )) {name=UVs CheapTri, nodeColor=<rgba 0.125 1 0 1>}
  n49 = NamedRerouteUsage()
  n50 = NamedRerouteUsage()
  NamedRerouteDeclaration(None=LinearInterpolate(
    A=ScalarParameter() {defaultValue=0.3, group=Wrap_HexColor, parameterName=HexC_Main_RoghnessHigh}
    B=ScalarParameter() {defaultValue=0.35, group=Wrap_HexColor, parameterName=HexC_Main_Roghness_Low}
    Alpha=TextureSample(UVs=@n49, Tex=@n50).R
   )) {name=Roughness, nodeColor=<rgba 0 1 0.8125 1>}
  n51 = ScalarParameter() {group=Weapon Specific Wrap Adjustments, parameterName=DO NOT OVERRIDE-Pickaxe-Position-BP-Switch, sortPriority=111}
  n52 = LinearInterpolate(A=Constant3Vector() {constant=<rgba 1 0 0 1>}, B=Constant3Vector() {constant=<rgba 0 1 0 1>}, Alpha=@n51)
  n53 = MaterialFunctionCall(Float4=NamedRerouteUsage()) {materialFunction=&BreakOutFloat4Components}
  n54 = LinearInterpolate(A=Constant3Vector() {constant=<rgba 0 0 1 1>}, B=Constant3Vector() {constant=<rgba 0 1 0 1>}, Alpha=@n53.G)
  n55 = Constant3Vector() {constant=<rgba 0 1 0 1>}
  n56 = Transform(None=LinearInterpolate(A=LinearInterpolate(A=@n52, B=@n54, Alpha=@n53.R), B=LinearInterpolate(A=@n55, B=@n55, Alpha=@n53.A), Alpha=@n53.B)) {transformSourceType=TRANSFORMSOURCE_Local}
  NamedRerouteDeclaration(None=@n56) {name=Constraint Axis, nodeColor=<rgba 0 0.9297 1 1>}
  n57 = Multiply(A=@n12.RGB, B=@n23.A)
  n58 = Saturate(None=@n22)
  n59 = NamedRerouteUsage()
  n60 = Constant() {r=0.5}
  n61 = MaterialFunctionCall(Float2=@n22) {materialFunction=&BreakOutFloat2Components}
  n62 = SphereMask(A=@n60, B=@n61.R) {attenuationRadius=0.5}
  n63 = SphereMask(A=@n61.G, B=@n60) {attenuationRadius=0.5}
  n64 = Multiply(
   A=@n21.Single_plane_Masking
   B=Multiply(A=TextureSample(UVs=@n58, Tex=@n59) {samplerSource=SSM_Clamp_WorldGroupSettings}.R, B=Multiply(A=@n62, B=@n63))
  )
  n65 = LinearInterpolate(A=NamedRerouteUsage(), B=NamedRerouteUsage(), Alpha=@n21.Single_plane_Masking on side) {constB=0}
  n66 = NamedRerouteUsage()
  n67 = Time()
  n68 = ScalarParameter() {defaultValue=0.1, group=Wrap_HexColor_Layers, parameterName=HexC_ColoCurve_Speed}
  n69 = Frac(None=Add(A=ComponentMask(None=@n66) {r=true}, B=Multiply(A=@n67, B=@n68)))
  n70 = CurveAtlasRowParameter(CurveTime=@n69) {
   atlas=/Game/Weapons/WeaponSkins/Wraps/Textures/CruveAtlas_Wraps_01
   curve=/Game/Weapons/WeaponSkins/Wraps/Textures/HunterD/CC_Rainbow
   defaultValue=42
   group=Wrap_HexColor_Layers
   parameterName=HexC_ColorCurve
  }
  n71 = MaterialFunctionCall!(
   Index=ScalarParameter() {defaultValue=42, group=Wrap_HexColor_Layers, parameterName="HexC_ColorCurve  Index"}
   CCinput=@n70
   CCinput_Alpha
   UV=@n69
  ) {materialFunction=&/MF_Global_CC_Tex}
  n72 = SetMaterialAttributes!(
   MaterialAttributes=NamedRerouteUsage()
   BaseColor=Constant()
   Roughness=ScalarParameter() {defaultValue=0.35, group=Wrap_HexColor, parameterName=HexC_Main_Roghness}
   Specular=ScalarParameter() {defaultValue=0.2, group=Wrap_HexColor, parameterName=HexC_Main_Specular}
   Metallic=ScalarParameter() {group=Wrap_HexColor, parameterName=HexC_Main_Metalic}
   EmissiveColor=Add(A=Multiply(A=@n57, B=@n64), B=Multiply(A=@n65, B=@n71.3V))
  ) {attributeSetTypes [
    #2
    #3
    #5
    #4
    #6
   ]}
  NamedRerouteDeclaration(None=@n72) {name=Material_Back, nodeColor=<rgba 0 0.3906 1 1>}
  NamedRerouteDeclaration(None=TextureObjectParameter() {group=Wrap_HexColor, parameterName=HexC_Tex_Hex_Warp, texture=&T_BlacksRGB}) {name=Tex_Hex_Warp, nodeColor=<rgba 0.2656 0 1 1>}
  NamedRerouteDeclaration(None=Multiply(
    A=MaterialFunctionCall() {materialFunction=&ObjectScale}.Scale Z
    B=PreSkinnedLocalBounds().Extents
   )) {name=Bounds_Extents, nodeColor=<rgba 0.5234 1 0 1>}
  n73 = FunctionInput(Preview=Constant3Vector() {constant=<rgba 0 1 0 0>}) {bUsePreviewValueAsDefault=true, inputName=Forward_Axis, sortPriority=4}
  n74 = VectorParameter() {group=Wrap_HexColor_2nd, parameterName=HexC_2nd_Roll_Forward_Axis, defaultValue=<rgba 0 1 0 1>}
  n75 = ScalarParameter() {group=Wrap_HexColor_2nd, parameterName=HexC_2nd_Roll_Is}
  n76 = MaterialFunctionCall(Float3=Multiply(A=@n1, B=LinearInterpolate(A=@n73, B=@n74.RGB, Alpha=@n75))) {materialFunction=&BreakOutFloat3Components}
  n77 = Max(A=Max(A=@n76.R, B=@n76.R), B=@n76.B)
  n78 = ScalarParameter() {defaultValue=0.5, group=Wrap_HexColor_2nd, parameterName=HexC_CC_Tile}
  n79 = Time()
  n80 = ScalarParameter() {defaultValue=0.05, group=Wrap_HexColor_2nd, parameterName=HexC_CC_Speed}
  n81 = Frac(None=Add(A=Multiply(A=@n77, B=@n78), B=Multiply(A=@n79, B=@n80)))
  n82 = CurveAtlasRowParameter(CurveTime=@n81) {
   atlas=&CruveAtlas_Wraps_01
   curve=&CC_Rainbow
   defaultValue=42
   group=Wrap_HexColor_Layers
   parameterName=HexC_ColorCurve
  }
  n83 = SetMaterialAttributes!(
   MaterialAttributes=NamedRerouteUsage()
   BaseColor=MaterialFunctionCall(Index=ScalarParameter() {defaultValue=42, group=Wrap_HexColor_Layers, parameterName="HexC_ColorCurve  Index"}, CCinput=@n82, UV=@n81) {materialFunction=&MF_Global_CC_Tex}.3V
   Roughness=ScalarParameter() {defaultValue=0.3, group=Wrap_HexColor_2nd, parameterName=HexC_2nd_Roughnes}
   Specular=ScalarParameter() {defaultValue=0.5, group=Wrap_HexColor_2nd, parameterName=HexC_2nd_Specular}
   Metallic=ScalarParameter() {group=Wrap_HexColor_2nd, parameterName=HexC_2nd_Metalic}
  ) {attributeSetTypes [
    #2
    #3
    #5
    #4
   ]}
  NamedRerouteDeclaration(None=@n83) {name=MAterial_Secondary, nodeColor=<rgba 0 0.3906 1 1>}
  n84 = ScalarParameter() {group=Wrap_HexColor_2nd_Uvs, parameterName=HexC_2nd_Rotation_Weapon}
  n85 = MaterialFunctionCall(Float4=NamedRerouteUsage()) {materialFunction=&BreakOutFloat4Components}
  n86 = LinearInterpolate(
   A=ScalarParameter() {group=Wrap_HexColor_2nd_Uvs, parameterName=HexC_2nd_Rotation_Vehicle}
   B=ScalarParameter() {group=Wrap_HexColor_2nd_Uvs, parameterName=HexC_2nd_Rotation_VehicleSide}
   Alpha=@n85.G
  )
  n87 = ScalarParameter() {defaultValue=0.25, group=Wrap_HexColor_2nd_Uvs, parameterName=HexC_2nd_Rotation_Mech}
  n88 = ScalarParameter() {group=Wrap_HexColor_2nd_Uvs, parameterName=HexC_2nd_Rotation_Character}
  NamedRerouteDeclaration(None=LinearInterpolate(A=LinearInterpolate(A=@n84, B=@n86, Alpha=@n85.R), B=LinearInterpolate(A=@n87, B=@n88, Alpha=@n85.A), Alpha=@n85.B)) {name=2nd_Rotation, nodeColor=<rgba 1 0.7266 0 1>}
  NamedRerouteDeclaration(None=Multiply(A=Time(), B=VectorParameter() {group=Wrap_HexColor_Layers, parameterName=HexC_Noise_Speed, defaultValue=<rgba 0.015 0.013 0 1>}.RGB)) {name=Noise_Speed, nodeColor=<rgba 1 0 0.3281 1>}
  n89 = Time()
  n90 = VectorParameter() {group=Wrap_HexColor_Layers, parameterName=HexC_Height_Speed, defaultValue=<rgba 0.05 0.03 0 1>}
  NamedRerouteDeclaration(None=ComponentMask(None=Multiply(A=@n89, B=@n90.RGB)) {g=true, r=true}) {name=Height_Speed, nodeColor=<rgba 1 0 0.3281 1>}
  n91 = FunctionInput(Preview=@n4.G) {bUsePreviewValueAsDefault=true, inputName=Mask Crevise, inputType=FunctionInput_Scalar, sortPriority=24, previewValue=<1 0 0 1>}
  n92 = ScalarParameter() {group=Wrap_HexColor, parameterName=HexC_Normal_warp}
  NamedRerouteDeclaration(None=ComponentMask(None=Multiply(A=@n91, B=@n92)) {g=true, r=true}) {name=Edge_Warp, nodeColor=<rgba 0 1 0.5547 1>}
  n93 = VectorParameter() {group=Wrap_HexColor_Layers, parameterName=HexC_Height_Amount_Vehicle, defaultValue=<rgba 0.5 1 1.5 1>}
  n94 = VectorParameter() {group=Wrap_HexColor_Layers, parameterName=HexC_Height_Amount_VehicleSide, defaultValue=<rgba 0.5 1 1.5 1>}
  n95 = MaterialFunctionCall(Float4=NamedRerouteUsage()) {materialFunction=&BreakOutFloat4Components}
  n96 = LinearInterpolate(
   A=LinearInterpolate(
    A=VectorParameter() {group=Wrap_HexColor_Layers, parameterName=HexC_Height_Amount_Weapon, defaultValue=<rgba 0.5 1 1.5 1>}.RGB
    B=LinearInterpolate(A=@n93.RGB, B=@n94.RGB, Alpha=@n95.G)
    Alpha=@n95.R
   )
   B=LinearInterpolate(
    A=VectorParameter() {group=Wrap_HexColor_Layers, parameterName=HexC_Height_Amoun_Mech, defaultValue=<rgba 0.5 1 1.5 1>}.RGB
    B=VectorParameter() {group=Wrap_HexColor_Layers, parameterName=HexC_Height_Amoun_Character, defaultValue=<rgba 0.5 1 1.5 1>}.RGB
    Alpha=@n95.A
   )
   Alpha=@n95.B
  )
  n97 = VectorParameter() {group=Wrap_HexColor_Layers, parameterName=HexC_Height_Amount_BackPack, defaultValue=<rgba 0.5 1 1.5 1>}
  n98 = MaterialFunctionCall(Float2=NamedRerouteUsage()) {materialFunction=&BreakOutFloat2Components}
  NamedRerouteDeclaration(None=Multiply(A=NamedRerouteUsage(), B=LinearInterpolate(A=@n96, B=@n97.RGB, Alpha=@n98.R))) {name=Height_Amount, nodeColor=<rgba 1 0.7266 0 1>}
  n99 = MaterialFunctionCall(Float4=NamedRerouteUsage()) {materialFunction=&BreakOutFloat4Components}
  n100 = LinearInterpolate(
   A=ScalarParameter() {defaultValue=1, group=Wrap_HexColor_Layers, parameterName=HexC_TileRate__Weapon}
   B=LinearInterpolate(
    A=ScalarParameter() {defaultValue=1, group=Wrap_HexColor_Layers, parameterName=HexC_TileRate_Vehicle}
    B=ScalarParameter() {defaultValue=1, group=Wrap_HexColor_Layers, parameterName=HexC_TileRate_VehicleSide}
    Alpha=@n99.G
   )
   Alpha=@n99.R
  )
  n101 = LinearInterpolate(
   A=ScalarParameter() {defaultValue=1, group=Wrap_HexColor_Layers, parameterName=HexC_TileRate_Mech}
   B=ScalarParameter() {defaultValue=1, group=Wrap_HexColor_Layers, parameterName=HexC_TileRate_Character}
   Alpha=@n99.A
  )
  n102 = NamedRerouteUsage()
  NamedRerouteDeclaration(None=LinearInterpolate(
    A=LinearInterpolate(A=@n100, B=@n101, Alpha=@n99.B)
    B=ScalarParameter() {defaultValue=1, group=Wrap_HexColor_Layers, parameterName=HexC_TileRate_BackPack}
    Alpha=MaterialFunctionCall(Float2=@n102) {materialFunction=&BreakOutFloat2Components}.R
   )) {name=Tile_Rate, nodeColor=<rgba 1 0.7266 0 1>}
  NamedRerouteDeclaration(None=@n77) {name=Forward axis, nodeColor=<rgba 1 0 0.1406 1>}
  NamedRerouteDeclaration(None=Normalize(VectorInput=@n56)) {name=SinglePlane_Normal, nodeColor=<rgba 0 0.9297 1 1>}
  NamedRerouteDeclaration(None=@n28.A) {name=Height_Noise, nodeColor=<rgba 1 0.0469 0 1>}
  n103 = ScalarParameter() {defaultValue=0.25, parameterName=HexC_Height_Speed_s}
  n104 = Add(A=Sine(None=Multiply(A=@n89, B=@n103))) {constB=0.5}
  NamedRerouteDeclaration(None=Saturate(None=Multiply(A=@n104) {constB=1.5})) {name=Height_Speed_1, nodeColor=<rgba 1 0 0.3281 1>}
  NamedRerouteDeclaration(None=FunctionInput(Preview=Constant4Vector().RGBA) {bUsePreviewValueAsDefault=true, inputName=IsState, inputType=FunctionInput_Vector4, sortPriority=4}) {name=IsState, nodeColor=<rgba 0.1497 1 0.1633 1>}
  NamedRerouteDeclaration(None=FunctionInput(Preview=Constant2Vector()) {bUsePreviewValueAsDefault=true, inputName=IsState_B, inputType=FunctionInput_Vector2, sortPriority=4}) {name=IsState_B, nodeColor=<rgba 0.1497 1 0.1633 1>}
  n105 = Subtract(A=@n40.PS, B=@n42.Min)
  n106 = MaterialFunctionCall(Color=@n43) {materialFunction=&SplitComponents}
  n107 = Max(A=Max(A=@n106.R, B=@n106.R), B=@n106.B)
  NamedRerouteDeclaration(None=FunctionInput(Preview=Divide(A=@n105, B=@n107)) {bUsePreviewValueAsDefault=true, inputName=Calcated_Bounds, sortPriority=11}) {name=Caclcated_Bounds, nodeColor=<rgba 0 0.9297 1 1>}
  n108 = Constant3Vector() {constant=<rgba 0 1 0 1>}
  n109 = Constant3Vector() {constant=<rgba 0 0 1 1>}
  n110 = Constant3Vector() {constant=<rgba 1 0 0 1>}
  n111 = Constant3Vector() {constant=<rgba 1 0 0 1>}
  n112 = MaterialFunctionCall(Float4=NamedRerouteUsage()) {materialFunction=&BreakOutFloat4Components}
  n113 = Constant3Vector() {constant=<rgba 0 0 1 1>}
  n114 = LinearInterpolate(
   A=LinearInterpolate(A=LinearInterpolate(A=@n108, B=@n109, Alpha=@n51), B=LinearInterpolate(A=@n110, B=@n111, Alpha=@n112.G), Alpha=@n112.R)
   B=LinearInterpolate(A=@n113, B=@n113, Alpha=@n112.A)
   Alpha=@n112.B
  )
  NamedRerouteDeclaration(None=Normalize(VectorInput=Transform(None=@n114) {transformSourceType=TRANSFORMSOURCE_Local})) {name=SinglePlane_X_Axis, nodeColor=<rgba 0 0.9297 1 1>}
  n115 = PreSkinnedNormal()
  NamedRerouteDeclaration(None=FunctionInput(Preview=VertexInterpolator(VS=@n115).PS) {bUsePreviewValueAsDefault=true, inputName=Pre-Skinned Local Normal, sortPriority=10}) {name=PreSkin_Normal, nodeColor=<rgba 0.8281 1 0 1>}
  NamedRerouteDeclaration(None=@n11.Normal) {name=Root_Normal, nodeColor=<rgba 0.4062 0 1 1>}
  NamedRerouteUsage()
  NamedRerouteUsage()
  NamedRerouteUsage()
  NamedRerouteUsage()
  NamedRerouteUsage()
  NamedRerouteUsage()
  NamedRerouteUsage()
  NamedRerouteUsage()
  NamedRerouteUsage()
  ScalarParameter() {defaultValue=0.25, group=Wrap_HexColor_Layers, parameterName=HexC_Height_Tilerate}
  Subtract(A=@n14.Max, B=@n14.Max)
 ]}
