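# Material function for weapon, vehicle and mech wraps that adds an iridescent holographic foil, toggled by the Wrap_HoloFoil static switch.
# Pre-skinned bounds position, distorted by world-aligned noise and shifted by camera direction, indexes a holo color curve atlas, with Fresnel-driven brightness.
# That color blends into base color, darkened by Fresnel outline and crevasse masks, while metallic and roughness are remapped.
# A sunlight-reflection chroma variant is present but not connected to the output.
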MaterialFunction /Game/Weapons/WeaponSkins/Wraps/MaterialFunctions/MF_Wrap_HoloFoil
graph {nodes [
  n1 = FunctionInput() {bUsePreviewValueAsDefault=true, inputName="Customization Mask (_CM B)", inputType=FunctionInput_Scalar, sortPriority=24}
  n2 = ScalarParameter() {group=Wrap_HoloFoil, parameterName=HF_Remove_Customization Mask, sortPriority=-10}
  BlendMaterialAttributes(Alpha=Saturate(None=Add(A=@n1, B=@n2))) {vertexAttributeBlendType=UseA}
  Constant() {r=0.18}
  Constant()
  n3 = FunctionInput() {inputName=Material, inputType=FunctionInput_MaterialAttributes, sortPriority=-100}
  n4 = GetMaterialAttributes(None=@n3) {attributeGetTypes [
    #1
    #2
    #3
    #4
    #5
   ]}
  n5 = OneMinus(None=Fresnel(ExponentIn=ScalarParameter() {defaultValue=4.724, group=Wrap_HoloFoil, parameterName=HF_FoilExponent}) {baseReflectFraction=0, exponent=2})
  n6 = ScalarParameter() {defaultValue=0.25, group=Wrap_HoloFoil, parameterName=HF_ChromaDistortionNoiseStrength}
  n7 = TextureObjectParameter() {
   group=Wrap_HoloFoil
   parameterName=HF_ChromaDistortionNoise
   samplerType=SAMPLERTYPE_LinearColor
   texture=/Game/Characters/Player/Female/Medium/Bodies/F_MED_HighTower_Squash/Textures/T_Swirly_Noise_highContrast1_Characters
  }
  n8 = VertexInterpolator(VS=PreSkinnedPosition())
  n9 = PreSkinnedLocalBounds()
  n10 = FunctionInput(Preview=Divide(A=Subtract(A=@n8.PS, B=@n9.Min), B=Subtract(A=@n9.Max, B=@n9.Max))) {bUsePreviewValueAsDefault=true, inputName=Calcated_Bounds, sortPriority=11}
  n11 = FunctionInput() {inputName=IsVehicle, inputType=FunctionInput_Scalar, sortPriority=4}
  n12 = LinearInterpolate(
   A=ScalarParameter() {defaultValue=1, group=Wrap_HoloFoil, parameterName=HF_ChromaDistortionNoise_Tiling_Weapon}
   B=ScalarParameter() {defaultValue=1, group=Wrap_HoloFoil, parameterName=HF_ChromaDistortionNoise_Tiling_Vehicle}
   Alpha=@n11
  )
  n13 = ScalarParameter() {defaultValue=1, group=Wrap_HoloFoil, parameterName=HF_ChromaDistortionNoise_Tiling_Mech}
  n14 = FunctionInput() {inputName=IsMech, inputType=FunctionInput_Scalar, sortPriority=4}
  n15 = PreSkinnedNormal()
  n16 = MaterialFunctionCall(
   TextureObject=@n7
   TextureSize=Constant() {r=1}
   WorldPosition=Multiply(A=@n10, B=LinearInterpolate(A=@n12, B=@n13, Alpha=@n14))
   World Space Normal=FunctionInput(Preview=VertexInterpolator(VS=@n15).PS) {bUsePreviewValueAsDefault=true, inputName=Pre-Skinned Local Normal, sortPriority=10}
  ) {materialFunction=&/WorldAlignedTexture}
  n17 = ComponentMask(None=Subtract(A=@n10, B=Multiply(A=@n6, B=@n16.XYZ Texture))) {b=true}
  n18 = MaterialFunctionCall() {materialFunction=&CameraDirectionVector}
  n19 = ScalarParameter() {defaultValue=0.5, group=Wrap_HoloFoil, parameterName=HF_ViewAnlgeFidelity}
  n20 = Multiply(A=DotProduct(A=Constant3Vector() {constant=<rgba 0 0 1 0>}, B=@n18.Result), B=@n19)
  n21 = Multiply(A=DotProduct(A=Constant3Vector() {constant=<rgba 1 0 0 0>}, B=@n18.Result), B=@n19)
  n22 = Add(A=@n20, B=@n21)
  n23 = Multiply(A=DotProduct(A=Constant3Vector() {constant=<rgba 0 1 0 0>}, B=@n18.Result), B=@n19)
  n24 = Add(A=ComponentMask(None=@n10) {r=true}, B=@n21)
  n25 = Add(A=ComponentMask(None=@n10) {g=true}, B=@n23)
  n26 = LinearInterpolate(
   A=Add(A=@n17, B=Add(A=@n22, B=@n23))
   B=Add(A=Add(A=@n24, B=@n25), B=Add(A=@n17, B=@n20))
   Alpha=ScalarParameter() {group=Wrap_HoloFoil, parameterName=HC_Use3Vector}
  )
  n27 = ScalarParameter() {defaultValue=1, group=Wrap_HoloFoil, parameterName=HF_Layer_01_ChromaOffset}
  n28 = CurveAtlasRowParameter(CurveTime=Frac(None=Add(A=@n26, B=@n27))) {
   atlas=&AthenaCurveAtlas01
   curve=/Game/Athena/LinearColorCurves/Curve_C_Holo
   defaultValue=250
   group=Wrap_HoloFoil
   parameterName=FoilChroma
  }
  n29 = Multiply(A=@n28, B=ScalarParameter() {defaultValue=0.5, group=Wrap_HoloFoil, parameterName=HF_Layer_01_Strength})
  n30 = Multiply(
   A=LinearInterpolate(
    A=ScalarParameter() {defaultValue=1, group=Wrap_HoloFoil, parameterName=HF_FoilMinBrightness}
    B=ScalarParameter() {defaultValue=7, group=Wrap_HoloFoil, parameterName=HF_FoilMaxBrightness}
    Alpha=Clamp(None=@n5)
   ) {constB=2}
   B=@n29
  )
  n31 = ScalarParameter() {defaultValue=1, group=Wrap_HoloFoil, parameterName=HF_BaseColorContribution}
  n32 = OneMinus(None=Fresnel(ExponentIn=ScalarParameter() {defaultValue=4.724, group=Wrap_HoloFoil, parameterName=HF_OutlineExponent1}) {baseReflectFraction=0})
  n33 = ScalarParameter() {defaultValue=4.639, group=Wrap_HoloFoil, parameterName=HF_OutlineExponent2}
  n34 = Saturate(None=LinearInterpolate(
    A=ScalarParameter() {defaultValue=-5.4358, group=Wrap_HoloFoil, parameterName=HF_OutlineC1}
    B=ScalarParameter() {defaultValue=41.451, group=Wrap_HoloFoil, parameterName=HF_OutlineC2}
    Alpha=Power(Base=@n32, Exp=@n33)
   ))
  n35 = ComponentMask(None=Saturate(None=Add(A=@n11, B=@n14))) {r=true}
  n36 = FunctionInput() {inputName="Customization Crevasse (_CM G)", inputType=FunctionInput_Scalar, sortPriority=21, previewValue=<1 0 0 1>}
  n37 = LinearInterpolate(
   A=LinearInterpolate(
    A=ScalarParameter() {defaultValue=10, group=Wrap_HoloFoil, parameterName=HF_Mask1}
    B=ScalarParameter() {defaultValue=10, group=Wrap_HoloFoil, parameterName=HF_Mask1_Vehicle}
    Alpha=@n35
   )
   B=LinearInterpolate(
    A=ScalarParameter() {defaultValue=-7, group=Wrap_HoloFoil, parameterName=HF_Mask3}
    B=ScalarParameter() {defaultValue=-7, group=Wrap_HoloFoil, parameterName=HF_Mask3_Vehicle}
    Alpha=@n35
   )
   Alpha=@n36
  )
  n38 = ScalarParameter() {group=Wrap_HoloFoil, parameterName=HF_ToonLine_Opacity}
  n39 = Add(A=Multiply(A=@n34, B=Saturate(None=@n37)), B=@n38)
  n40 = ScalarParameter() {defaultValue=0.1, group=Wrap_HoloFoil, parameterName=HF_Roughness_Min}
  n41 = ScalarParameter() {defaultValue=0.2, group=Wrap_HoloFoil, parameterName=HF_Roughness_Max}
  n42 = ShadingPathSwitch(Default=QualitySwitch(Default=LinearInterpolate(A=@n40, B=@n41, Alpha=@n4.Roughness)), Forward=@n41, Mobile=@n41)
  n43 = LinearInterpolate(
   A=ScalarParameter() {defaultValue=1, group=Wrap_HoloFoil, parameterName=HF_Metalness_Min}
   B=ScalarParameter() {defaultValue=1, group=Wrap_HoloFoil, parameterName=HF_Metalness_Max}
   Alpha=@n4.Metallic
  )
  n44 = SetMaterialAttributes!(
   MaterialAttributes=@n4.MaterialAttributes
   Metallic=@n43
   Roughness=@n42
   BaseColor=Multiply(A=LinearInterpolate(A=@n4.BaseColor, B=@n30, Alpha=@n31), B=Saturate(None=@n39))
   Normal=@n4.MaterialAttributes
  ) {attributeSetTypes [
    #3
    #4
    #5
    #1
   ]}
  FunctionOutput(None=StaticSwitchParameter(True=@n44, False=@n3) {group=Customization_Wrap, parameterName=Wrap_HoloFoil})
  n45 = Multiply(A=@n28, B=ScalarParameter() {defaultValue=1, group=Wrap_HoloFoil, parameterName=HF_MobileLayer_Strength})
  n46 = ShadingPathSwitch(Default=QualitySwitch(Default=@n29), Forward=@n45, Mobile=@n45)
  LinearInterpolate(A=Multiply(A=@n46, B=ScalarParameter() {group=Wrap_HoloFoil, parameterName=HF_Layer_00_Strength}), B=@n46)
  LinearInterpolate(A=@n4.Roughness, B=@n42)
  LinearInterpolate(B=@n34, Alpha=@n38) {constA=1}
  LinearInterpolate(A=@n4.Metallic, B=@n43)
  n47 = ScalarParameter() {defaultValue=0.5, group=Wrap_HoloFoil, parameterName=HF_Layer_02_ChromaOffset}
  n48 = ScalarParameter() {group=Wrap_HoloFoil, parameterName=HF_Layer_03_ChromaOffset}
  n49 = Add(A=Multiply(A=@n27), B=Add(A=Multiply(A=@n47), B=Multiply(A=@n48)))
  n50 = Multiply(A=ScalarParameter() {defaultValue=0.5, group=Wrap_HoloFoil, parameterName=HF_MobileChromaOffset})
  Multiply(A=ShadingPathSwitch(Default=QualitySwitch(Default=@n49), Forward=@n50, Mobile=@n50), B=@n19)
  n51 = Constant()
  n52 = ScalarParameter() {defaultValue=3, group=Wrap_HoloFoil, parameterName=HF_ChromaReflection}
  n53 = Multiply(A=OneMinus(None=Fresnel(ExponentIn=@n52) {baseReflectFraction=0})) {constB=2}
  n54 = CollectionParameter() {
   collection=&FortniteMaterialParameters
   parameterName=SunlightDirectionalVector
  }
  n55 = ComponentMask(None=@n54) {b=true, g=true, r=true}
  n56 = ComponentMask(None=PixelNormalWS()) {b=true, g=true, r=true}
  n57 = LinearInterpolate(
   A=@n56
   B=ComponentMask(None=VertexNormalWS()) {b=true, g=true, r=true}
   Alpha=ScalarParameter() {defaultValue=0.25, group=Wrap_HoloFoil, parameterName=HF_Pixel-Vertex}
  )
  n58 = ConstantBiasScale(None=DotProduct(A=ComponentMask(None=@n55) {b=true, g=true, r=true}, B=ShadingPathSwitch(Default=@n57, Forward=@n56, Mobile=@n56))) {scale=0.28}
  n59 = CollectionParameter() {
   collection=&FortniteMaterialParameters
   parameterName=SunlightDirectionalVector
  }
  n60 = ComponentMask(None=@n59) {b=true, g=true, r=true}
  n61 = PixelNormalWS()
  n62 = Multiply(A=DotProduct(A=@n60, B=@n61)) {constB=2}
  n63 = Normalize(VectorInput=Subtract(A=Multiply(A=@n62, B=@n61), B=@n60))
  n64 = Multiply(A=MaterialFunctionCall() {materialFunction=&CameraDirectionVector}.Result) {constB=-1}
  n65 = Power(
   Base=Max(A=DotProduct(A=@n63, B=@n64)) {constB=0}
   Exp=ScalarParameter() {defaultValue=3, group=Wrap_HoloFoil, parameterName=HF_Chroma_Spec_Size}
  )
  n66 = Add(A=Multiply(A=Saturate(None=@n53), B=Clamp(None=@n58) {minDefault=0.25}), B=Saturate(None=Multiply(A=@n65) {constB=0.2}))
  n67 = ScalarParameter() {group=Wrap_HoloFoil, parameterName=HF_ChromaOffset}
  n68 = CurveAtlasRowParameter(CurveTime=Frac(None=Add(A=@n66, B=@n67))) {
   atlas=&AthenaCurveAtlas01
   curve=&Curve_C_Holo
   defaultValue=249
   group=Wrap_HoloFoil
   parameterName=FoilChroma
  }
  SetMaterialAttributes!(MaterialAttributes=@n4.MaterialAttributes, Metallic=@n51, Roughness, BaseColor=@n51, EmissiveColor=@n68) {attributeSetTypes [
    #3
    #4
    #5
    #2
   ]}
  StaticSwitchParameter(True=@n68, False=@n30) {group=Wrap_HoloFoil, parameterName="HF_Object--> lighting Based. ", sortPriority=-10}
  VertexNormalWS()
 ]}
