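# DinoHunter wrap material function that layers a textured, reactive glowing skin over an input material through customization masks, gated by a static switch.
# IsState picks projection settings for weapon, vehicle, vehicle side, mech or character.
# Projected textures drive color, roughness, metallic and normal, with a scrolling curve-atlas gradient in masked areas.
# Mirrored side projections shape radial emissive gradients, revealed by three overlapping pulses timed from Blueprint-named timestamp parameters.
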MaterialFunction /Game/Weapons/WeaponSkins/Wraps/MaterialFunctions/MF_Wrap_HunterD
graph {nodes [
  Constant3Vector()
  Constant()
  FunctionInput(Preview=TextureSample() {samplerType=SAMPLERTYPE_Masks, texture=/Game/Weapons/FORT_Rifles/Textures/T_Scar_CM}.G) {bUsePreviewValueAsDefault=true, inputName=Curavture, inputType=FunctionInput_Scalar, sortPriority=24}
  FunctionInput(Preview=Constant3Vector() {constant=<rgba 0 1 0 0>}) {bUsePreviewValueAsDefault=true, inputName=Forward Axis}
  n1 = FunctionInput(Preview=SetMaterialAttributes!(MaterialAttributes, Normal=TextureSample() {samplerType=SAMPLERTYPE_Normal, texture=/Game/Weapons/FORT_Rifles/T_SCAR_N}.RGB) {attributeSetTypes [#1]}) {bUsePreviewValueAsDefault=true, inputName=Material, inputType=FunctionInput_MaterialAttributes, sortPriority=-20}
  n2 = GetMaterialAttributes(None=@n1) {attributeGetTypes [#1]}
  n3 = TextureObjectParameter() {
   group=Wrap_DinoHunter
   parameterName=DinoH__Text_Normal
   samplerType=SAMPLERTYPE_Normal
   texture=&FlatNormalBlue
  }
  n4 = FunctionInput(Preview=Constant()) {bUsePreviewValueAsDefault=true, inputName=IsState, inputType=FunctionInput_Vector4}
  n5 = MaterialFunctionCall(Float4=@n4) {materialFunction=&BreakOutFloat4Components}
  n6 = ScalarParameter() {group=Wrap_DinoHunter, parameterName=DinoH_UV_Rotation_Weapon}
  n7 = MaterialFunctionCall(Float4=@n4) {materialFunction=&BreakOutFloat4Components}
  n8 = LinearInterpolate(
   A=ScalarParameter() {group=Wrap_DinoHunter, parameterName=DinoH_UV_Rotation_Vehicle}
   B=ScalarParameter() {group=Wrap_DinoHunter, parameterName=DinoH_UV_Rotation_VehicleSide}
   Alpha=@n7.G
  )
  n9 = ScalarParameter() {defaultValue=0.25, group=Wrap_DinoHunter, parameterName=DinoH_UV_Rotation_Mech}
  n10 = ScalarParameter() {group=Wrap_DinoHunter, parameterName=DinoH_UV_Rotation_Character}
  n11 = VectorParameter() {group=Wrap_DinoHunter, parameterName=DinoH_UV_Offset_Weapon, defaultValue=<rgba 0 0 0 1>}
  n12 = MaterialFunctionCall(Float4=@n4) {materialFunction=&BreakOutFloat4Components}
  n13 = LinearInterpolate(
   A=VectorParameter() {group=Wrap_DinoHunter, parameterName=DinoH_UV_Offset_Vehicle, defaultValue=<rgba 0 0 0 1>}.RGB
   B=VectorParameter() {group=Wrap_DinoHunter, parameterName=DinoH_UV_Offset_VehicleSide, defaultValue=<rgba 0 0 0 1>}.RGB
   Alpha=@n12.G
  )
  n14 = VectorParameter() {group=Wrap_DinoHunter, parameterName=DinoH_UV_Offset_Mech, defaultValue=<rgba 0 0 0 1>}
  n15 = VectorParameter() {group=Wrap_DinoHunter, parameterName=DinoH_UV_Offset_Character, defaultValue=<rgba 0 0 0 1>}
  n16 = ScalarParameter() {defaultValue=1, group=Wrap_DinoHunter, parameterName=DinoH_UV_TileRate_Weapon}
  n17 = MaterialFunctionCall(Float4=@n4) {materialFunction=&BreakOutFloat4Components}
  n18 = LinearInterpolate(
   A=ScalarParameter() {defaultValue=1, group=Wrap_DinoHunter, parameterName=DinoH_UV_TileRate_Vehicle}
   B=ScalarParameter() {defaultValue=1, group=Wrap_DinoHunter, parameterName=DinoH_UV_TileRate_VehicleSide}
   Alpha=@n17.G
  )
  n19 = ScalarParameter() {defaultValue=1, group=Wrap_DinoHunter, parameterName=DinoH_UV_TileRate_Mech}
  n20 = ScalarParameter() {defaultValue=1, group=Wrap_DinoHunter, parameterName=DinoH_UV_TileRate_Character}
  n21 = FunctionInput(Preview=VertexInterpolator(VS=PreSkinnedNormal()).PS) {bUsePreviewValueAsDefault=true, inputName=Pre-Skinned Local Normal, sortPriority=10}
  n22 = VertexInterpolator(VS=PreSkinnedPosition())
  n23 = PreSkinnedLocalBounds()
  n24 = FunctionInput(Preview=Divide(A=Subtract(A=@n22.PS, B=@n23.Min), B=Subtract(A=@n23.Max, B=@n23.Max))) {bUsePreviewValueAsDefault=true, inputName=Calcated_Bounds, sortPriority=11}
  n25 = MaterialFunctionCall!(
   For_Safe_Zones
   Normal_Map=@n3
   Caculated_Bounds=@n24
   Pre_Skinned_Normals=@n21
   Is Vehicle=@n5.R
   Is VehicleSide=@n5.R
   Is Mech=@n5.R
   Vehicle_Rotation
   UV_Rotation Bypass=LinearInterpolate(A=LinearInterpolate(A=@n6, B=@n8, Alpha=@n7.R), B=LinearInterpolate(A=@n9, B=@n10, Alpha=@n7.A), Alpha=@n7.B)
   Mech_Rotation
   Uvs_Offset=LinearInterpolate(A=LinearInterpolate(A=@n11.RGB, B=@n13, Alpha=@n12.R), B=LinearInterpolate(A=@n14.RGB, B=@n15.RGB, Alpha=@n12.A), Alpha=@n12.B)
   Offset-after rotation
   Uvs_TileRate=LinearInterpolate(A=LinearInterpolate(A=@n16, B=@n18, Alpha=@n17.R), B=LinearInterpolate(A=@n19, B=@n20, Alpha=@n17.A), Alpha=@n17.B)
  ) {materialFunction=&/MF_Projection_Side_Normal}
  n26 = TextureObjectParameter() {group=Wrap_DinoHunter, parameterName=DinoH__Text_M, texture=&T_BlacksRGB}
  n27 = TextureSample(UVs=@n25.UVs, Tex=@n26)
  n28 = TextureSample(
   UVs=@n25.UVs
   Tex=TextureObjectParameter() {group=Wrap_DinoHunter, parameterName=DinoH__Text_BC, texture=&T_BlacksRGB}
  )
  n29 = LinearInterpolate(
   A=ScalarParameter() {group=Wrap_DinoHunter, parameterName=DinoH_Roghness_Low, sortPriority=0}
   B=ScalarParameter() {defaultValue=1, group=Wrap_DinoHunter, parameterName=DinoH_Roghness_High, sortPriority=0}
   Alpha=@n28.A
  )
  n30 = ScalarParameter() {defaultValue=0.5, group=Wrap_DinoHunter, parameterName=DinoH_Roghness_Stripes, sortPriority=0}
  n31 = TextureSample(UVs=@n25.UVs, Tex=@n26)
  n32 = MaterialFunctionCall(UVs=@n25.UVs) {materialFunction=&ScaleUVsByCenter}
  n33 = Multiply(A=@n31.A, B=@n32.0-1 mask)
  n34 = Constant() {r=0.5}
  n35 = MaterialFunctionCall(Float2=@n25.UVs) {materialFunction=&BreakOutFloat2Components}
  n36 = SphereMask(A=@n34, B=@n35.R) {attenuationRadius=0.5}
  n37 = SphereMask(A=@n35.G, B=@n34) {attenuationRadius=0.5}
  n38 = MaterialFunctionCall(Color=Multiply(A=@n27.RGB, B=Multiply(A=@n36, B=@n37))) {materialFunction=&SplitComponents}
  n39 = Multiply(A=@n33, B=OneMinus(None=@n38.G))
  n40 = ScalarParameter() {defaultValue=1, group=Wrap_DinoHunter, parameterName=DinoH_Grad_BackGround_TileRate, sortPriority=120}
  n41 = ScalarParameter() {defaultValue=1, group=Wrap_DinoHunter, parameterName=DinoH_Grad_Size_Vehicle}
  n42 = ScalarParameter() {defaultValue=1, group=Wrap_DinoHunter, parameterName=DinoH_Grad_Size_VehicleSide}
  n43 = MaterialFunctionCall(Color=@n4) {materialFunction=&SplitComponents}
  n44 = LinearInterpolate(
   A=LinearInterpolate(
    A=ScalarParameter() {defaultValue=1, group=Wrap_DinoHunter, parameterName=DinoH_Grad_Size_Weapon}
    B=LinearInterpolate(A=@n41, B=@n42, Alpha=@n43.G)
    Alpha=@n43.R
   )
   B=ScalarParameter() {defaultValue=1, group=Wrap_DinoHunter, parameterName=DinoH_Grad_Size_Mech}
   Alpha=@n43.B
  )
  n45 = MaterialFunctionCall(Float4=@n4) {materialFunction=&BreakOutFloat4Components}
  n46 = ScalarParameter() {group=Wrap_DinoHunter, parameterName=DinoH_MAsk_Rotation_Right_Vehicle}
  n47 = ScalarParameter() {group=Wrap_DinoHunter, parameterName=DinoH_MAsk_Rotation_Right_VehicleSide}
  n48 = MaterialFunctionCall(Color=@n4) {materialFunction=&SplitComponents}
  n49 = LinearInterpolate(
   A=LinearInterpolate(
    A=ScalarParameter() {group=Wrap_DinoHunter, parameterName=DinoH_MAsk_Rotation_Right_Weapon}
    B=LinearInterpolate(A=@n46, B=@n47, Alpha=@n48.G)
    Alpha=@n48.R
   )
   B=ScalarParameter() {defaultValue=0.25, group=Wrap_DinoHunter, parameterName=DinoH_MAsk_Rotation_Right_Mech}
   Alpha=@n48.B
  )
  n50 = VectorParameter() {group=Wrap_DinoHunter, parameterName=DinoH_Mask_Offset_Right_Vehicle, defaultValue=<rgba 0 0 0 1>}
  n51 = VectorParameter() {group=Wrap_DinoHunter, parameterName=DinoH_Mask_Offset_Right_VehicleSide, defaultValue=<rgba 0 0 0 1>}
  n52 = MaterialFunctionCall(Color=@n4) {materialFunction=&SplitComponents}
  n53 = LinearInterpolate(
   A=LinearInterpolate(
    A=VectorParameter() {group=Wrap_DinoHunter, parameterName=DinoH_Mask_Offset_Right_Weapon, defaultValue=<rgba 0 0 0 1>}.RGB
    B=LinearInterpolate(A=@n50.RGB, B=@n51.RGB, Alpha=@n52.G)
    Alpha=@n52.R
   )
   B=VectorParameter() {group=Wrap_DinoHunter, parameterName=DinoH_Mask_Offset_Right_Mech, defaultValue=<rgba 0 0 0 1>}.RGB
   Alpha=@n52.B
  )
  n54 = ScalarParameter() {defaultValue=1, group=Wrap_DinoHunter, parameterName=DinoH_Mask_Scale_Right_Vehicle}
  n55 = ScalarParameter() {defaultValue=1, group=Wrap_DinoHunter, parameterName=DinoH_Mask_Scale_Right_VehicleSide}
  n56 = MaterialFunctionCall(Color=@n4) {materialFunction=&SplitComponents}
  n57 = LinearInterpolate(
   A=LinearInterpolate(
    A=ScalarParameter() {defaultValue=1, group=Wrap_DinoHunter, parameterName=DinoH_Mask_Scale_Right_Weapon}
    B=LinearInterpolate(A=@n54, B=@n55, Alpha=@n56.G)
    Alpha=@n56.R
   )
   B=ScalarParameter() {defaultValue=1, group=Wrap_DinoHunter, parameterName=DinoH_Mask_Scale_Right_Mech}
   Alpha=@n56.B
  )
  n58 = MaterialFunctionCall!(
   Normal_Map
   Caculated_Bounds=@n24
   Pre_Skinned_Normals=@n21
   Is Vehicle=@n45.R
   Is Vehicle side=@n45.R
   Is Mech=@n45.R
   Offset-after rotation
   UV_Rotation_Right=@n49
   UV_Rotation_Left=Multiply(A=@n49) {constB=-1}
   Uvs_TileRate_Right=@n57
   Uvs_TileRate_Left=@n57
   Uvs_Offset_Right=@n53
   Uvs_Offset_Left=Multiply(A=@n53, B=Constant3Vector() {constant=<rgba -1 1 1 0>})
  ) {materialFunction=&/MF_Projection_Side_Left_Right_Decal_Normal}
  n59 = SphereMask(A=@n58.UVs, B=Constant() {r=0.5}, Radius=@n44, Hardness=Constant()) {hardnessPercent=0}
  n60 = Time()
  n61 = ScalarParameter() {defaultValue=0.5, group=Wrap_DinoHunter, parameterName=DinoH_React_Gradent_Speed, sortPriority=100}
  n62 = Frac(None=Add(A=Multiply(A=@n40, B=@n59), B=Multiply(A=@n60, B=@n61)))
  n63 = CurveAtlasRowParameter(CurveTime=@n62) {
   atlas=/Game/Weapons/WeaponSkins/Wraps/Textures/CruveAtlas_Wraps_01
   curve=/Game/Weapons/WeaponSkins/Wraps/Textures/NeonBands/T_RainbowLavaB
   defaultValue=17
   group=Wrap_DinoHunter
   parameterName=DinoH_Grad_CC
  }
  n64 = ScalarParameter() {defaultValue=41, group=Wrap_DinoHunter, parameterName="DinoH_Grad_Pulse_CC  Index"}
  n65 = ScalarParameter() {defaultValue=1, group=Wrap_DinoHunter, parameterName=DinoH_React_Gradent_Pulse_Start, sortPriority=100}
  n66 = ScalarParameter() {group=Wrap_DinoHunter, parameterName=DinoH_React_Gradent_Pulse_End, sortPriority=100}
  n67 = Time()
  n68 = ScalarParameter() {defaultValue=-9999999, group=Wrap_DinoHunter, parameterName=DinoH_TimeFrom_BP_A, sortPriority=99}
  n69 = ScalarParameter() {defaultValue=0.25, group=Wrap_DinoHunter, parameterName=DinoH_React_Gradent_Duration, sortPriority=100}
  n70 = Saturate(None=Divide(A=Subtract(A=@n67, B=@n68), B=@n69))
  n71 = CurveAtlasRowParameter(CurveTime=Add(A=@n59, B=LinearInterpolate(A=@n65, B=@n66, Alpha=@n70))) {
   atlas=&CruveAtlas_Wraps_01
   curve=/Game/Weapons/WeaponSkins/Wraps/Textures/HunterD/CC_Pulse_A
   defaultValue=41
   group=Wrap_DinoHunter
   parameterName=DinoH_Grad_Pulse_CC
  }
  n72 = ComponentMask(None=MaterialFunctionCall!(Index=@n64, CCinput=@n71, CCinput_Alpha, UV) {materialFunction=&/MF_Global_CC_Tex}.3V) {r=true}
  n73 = Constant() {r=0.5}
  n74 = Constant()
  n75 = Time()
  n76 = ScalarParameter() {defaultValue=-9999999, group=Wrap_DinoHunter, parameterName=DinoH_TimeFrom_BP_B, sortPriority=99}
  n77 = Saturate(None=Divide(A=Subtract(A=@n75, B=@n76), B=@n69))
  n78 = CurveAtlasRowParameter(CurveTime=Add(A=SphereMask(A=@n58.UVs, B=@n73, Radius=@n44, Hardness=@n74) {hardnessPercent=0}, B=LinearInterpolate(A=@n65, B=@n66, Alpha=@n77))) {
   atlas=&CruveAtlas_Wraps_01
   curve=&CC_Pulse_A
   defaultValue=41
   group=Wrap_DinoHunter
   parameterName=DinoH_Grad_Pulse_CC
  }
  n79 = ComponentMask(None=MaterialFunctionCall(Index=@n64, CCinput=@n78) {materialFunction=&MF_Global_CC_Tex}.3V) {r=true}
  n80 = Constant() {r=0.5}
  n81 = Constant()
  n82 = Time()
  n83 = ScalarParameter() {defaultValue=-9999999, group=Wrap_DinoHunter, parameterName=DinoH_TimeFrom_BP_C, sortPriority=99}
  n84 = Saturate(None=Divide(A=Subtract(A=@n82, B=@n83), B=@n69))
  n85 = CurveAtlasRowParameter(CurveTime=Add(A=SphereMask(A=@n58.UVs, B=@n80, Radius=@n44, Hardness=@n81) {hardnessPercent=0}, B=LinearInterpolate(A=@n65, B=@n66, Alpha=@n84))) {
   atlas=&CruveAtlas_Wraps_01
   curve=&CC_Pulse_A
   defaultValue=41
   group=Wrap_DinoHunter
   parameterName=DinoH_Grad_Pulse_CC
  }
  n86 = MaterialFunctionCall(Index=@n64, CCinput=@n85.R) {materialFunction=&MF_Global_CC_Tex}
  n87 = Saturate(None=Add(A=Add(A=@n72, B=@n79), B=ComponentMask(None=@n86.3V) {r=true}))
  n88 = Multiply(
   A=MaterialFunctionCall(Index=ScalarParameter() {defaultValue=17, group=Wrap_DinoHunter, parameterName="DinoH_Grad_CC  Index"}, CCinput=@n63, UV=@n62) {materialFunction=&MF_Global_CC_Tex}.3V
   B=@n87
  )
  n89 = ScalarParameter() {defaultValue=1, group=Wrap_DinoHunter, parameterName=DinoH_Grad_BackGround_Brightness, sortPriority=120}
  n90 = Add(A=OneMinus(None=@n31.A), B=LinearInterpolate(Alpha=@n87) {constA=0.75, constB=0}) {constB=0.75}
  n91 = Multiply(A=@n28.RGB, B=LinearInterpolate(B=Saturate(None=@n90), Alpha=@n32.0-1 mask) {constA=1})
  n92 = OneMinus(None=@n38.G)
  n93 = ScalarParameter() {defaultValue=1, group=Wrap_DinoHunter, parameterName=DinoH_Grad_Text_Brightness}
  n94 = MaterialFunctionCall(Float4=@n4) {materialFunction=&BreakOutFloat4Components}
  n95 = LinearInterpolate(
   A=ScalarParameter() {group=Wrap_DinoHunter, parameterName=DinoH_Grad_Rotation_Weapon}
   B=LinearInterpolate(
    A=ScalarParameter() {group=Wrap_DinoHunter, parameterName=DinoH_Grad_Rotation_Vehicle}
    B=ScalarParameter() {group=Wrap_DinoHunter, parameterName=DinoH_Grad_Rotation_VehicleSide}
    Alpha=@n94.G
   )
   Alpha=@n94.R
  )
  n96 = LinearInterpolate(
   A=ScalarParameter() {group=Wrap_DinoHunter, parameterName=DinoH_Grad_Rotation_Mech}
   B=ScalarParameter() {group=Wrap_DinoHunter, parameterName=DinoH_Grad_Rotation_Character}
   Alpha=@n94.A
  )
  n97 = ComponentMask(None=MaterialFunctionCall(UVs=@n25.UVs No Offsets/Tiling, "Rotation Angle (0-1)"=LinearInterpolate(A=@n95, B=@n96, Alpha=@n94.B)) {materialFunction=&CustomRotator}.Rotated Values) {r=true}
  n98 = ScalarParameter() {defaultValue=1, group=Wrap_DinoHunter, parameterName=DinoH_Grad_TileRate_Vehicle}
  n99 = ScalarParameter() {defaultValue=1, group=Wrap_DinoHunter, parameterName=DinoH_Grad_TileRate_VehicleSide}
  n100 = MaterialFunctionCall(Float4=@n4) {materialFunction=&BreakOutFloat4Components}
  n101 = LinearInterpolate(
   A=LinearInterpolate(
    A=ScalarParameter() {defaultValue=1, group=Wrap_DinoHunter, parameterName=DinoH_Grad_TileRate_Weapon}
    B=LinearInterpolate(A=@n98, B=@n99, Alpha=@n100.G)
    Alpha=@n100.R
   )
   B=LinearInterpolate(
    A=ScalarParameter() {defaultValue=1, group=Wrap_DinoHunter, parameterName=DinoH_Grad_TileRate_Mech}
    B=ScalarParameter() {defaultValue=1, group=Wrap_DinoHunter, parameterName=DinoH_Grad_TileRate_Character}
    Alpha=@n100.A
   )
   Alpha=@n100.B
  )
  n102 = Time()
  n103 = ScalarParameter() {group=Wrap_DinoHunter, parameterName=DinoH_Grad_Speed}
  n104 = Frac(None=Add(A=Multiply(A=@n97, B=@n101), B=Multiply(A=@n102, B=@n103)))
  n105 = CurveAtlasRowParameter(CurveTime=@n104) {
   atlas=&CruveAtlas_Wraps_01
   curve=&T_RainbowLavaB
   defaultValue=17
   group=Wrap_DinoHunter
   parameterName=DinoH_Grad_CC
  }
  n106 = MaterialFunctionCall(Index=ScalarParameter() {defaultValue=17, group=Wrap_DinoHunter, parameterName="DinoH_Grad_CC  Index"}, CCinput=@n105, UV=@n104) {materialFunction=&MF_Global_CC_Tex}
  n107 = SetMaterialAttributes!(
   MaterialAttributes=@n2.MaterialAttributes
   BaseColor=LinearInterpolate(A=Multiply(A=@n91, B=@n92), B=Multiply(A=@n93, B=@n106.3V), Alpha=@n38.R)
   Roughness=LinearInterpolate(
    A=LinearInterpolate(A=@n29, B=@n30, Alpha=@n33)
    B=ScalarParameter() {defaultValue=0.5, group=Wrap_DinoHunter, parameterName=DinoH_Roghness_Text, sortPriority=0}
    Alpha=@n38.R
   )
   Specular=ScalarParameter() {defaultValue=0.5, group=Wrap_DinoHunter, parameterName=DinoH_Specular, sortPriority=0}
   EmissiveColor=Multiply(A=Saturate(None=@n39), B=Multiply(A=@n88, B=@n89))
   Metallic=@n27.B
   Normal=Add(A=@n2.Normal, B=@n25.Raw_Normal)
  ) {attributeSetTypes [
    #2
    #3
    #4
    #5
    #6
    #1
   ]}
  n108 = TextureSample() {samplerType=SAMPLERTYPE_Masks, texture=&T_Scar_CM}
  n109 = NamedRerouteUsage()
  n110 = NamedRerouteUsage()
  n111 = LinearInterpolate(
   A=FunctionInput(Preview=OneMinus(None=@n108.B)) {bUsePreviewValueAsDefault=true, inputName=Customization Mask A, inputType=FunctionInput_Scalar, sortPriority=24}
   B=FunctionInput() {bUsePreviewValueAsDefault=true, inputName=Customization Mask B, inputType=FunctionInput_Scalar, sortPriority=24}
   Alpha=Saturate(None=Subtract(A=@n109, B=@n110))
  )
  FunctionOutput(None=StaticSwitchParameter(True=BlendMaterialAttributes(A=@n1, B=@n107, Alpha=@n111) {vertexAttributeBlendType=UseA}, False=@n1) {group=Customization_Wrap, parameterName=Wrap_DinoHunter})
  n112 = VectorParameter() {group=Wrap_DinoHunter, parameterName=DinoH_Mask_Offset_Left_Vehicle, defaultValue=<rgba 0 0 0 1>}
  n113 = VectorParameter() {group=Wrap_DinoHunter, parameterName=DinoH_Mask_Offset_Left_VehicleSide, defaultValue=<rgba 0 0 0 1>}
  n114 = MaterialFunctionCall(Color=@n4) {materialFunction=&SplitComponents}
  LinearInterpolate(
   A=LinearInterpolate(
    A=VectorParameter() {group=Wrap_DinoHunter, parameterName=DinoH_Mask_Offset_Left_Weapon, defaultValue=<rgba 0 0 0 1>}.RGB
    B=LinearInterpolate(A=@n112.RGB, B=@n113.RGB, Alpha=@n114.G)
    Alpha=@n114.R
   )
   B=VectorParameter() {group=Wrap_DinoHunter, parameterName=DinoH_Mask_Offset_Left_Mech, defaultValue=<rgba 0 0 0 1>}.RGB
   Alpha=@n114.B
  )
  n115 = ScalarParameter() {defaultValue=1, group=Wrap_DinoHunter, parameterName=DinoH_Mask_Scale_Left_Vehicle}
  n116 = ScalarParameter() {defaultValue=1, group=Wrap_DinoHunter, parameterName=DinoH_Mask_Scale_Left_VehicleSide}
  n117 = MaterialFunctionCall(Color=@n4) {materialFunction=&SplitComponents}
  LinearInterpolate(
   A=LinearInterpolate(
    A=ScalarParameter() {defaultValue=1, group=Wrap_DinoHunter, parameterName=DinoH_Mask_Scale_Left_Weapon}
    B=LinearInterpolate(A=@n115, B=@n116, Alpha=@n117.G)
    Alpha=@n117.R
   )
   B=ScalarParameter() {defaultValue=1, group=Wrap_DinoHunter, parameterName=DinoH_Mask_Scale_Left_Mech}
   Alpha=@n117.B
  )
  n118 = ScalarParameter() {group=Wrap_DinoHunter, parameterName=DinoH_MAsk_Rotation_Left_Vehicle}
  n119 = ScalarParameter() {group=Wrap_DinoHunter, parameterName=DinoH_MAsk_Rotation_Left_VehicleSide}
  n120 = MaterialFunctionCall(Color=@n4) {materialFunction=&SplitComponents}
  LinearInterpolate(
   A=LinearInterpolate(
    A=ScalarParameter() {group=Wrap_DinoHunter, parameterName=DinoH_MAsk_Rotation_Left_Weapon}
    B=LinearInterpolate(A=@n118, B=@n119, Alpha=@n120.G)
    Alpha=@n120.R
   )
   B=ScalarParameter() {defaultValue=0.25, group=Wrap_DinoHunter, parameterName=DinoH_MAsk_Rotation_Left_Mech}
   Alpha=@n120.B
  )
  MaterialFunctionCall() {materialFunction=&ScaleUVsByCenter}
  NamedRerouteDeclaration(None=MaterialFunctionCall(Float4=@n4) {materialFunction=&BreakOutFloat4Components}.A) {name=IsCharacter, nodeColor=<rgba 0 0.9062 1 1>}
  NamedRerouteDeclaration(None=FunctionInput() {bUsePreviewValueAsDefault=true, inputName=isCharAccent, inputType=FunctionInput_Scalar, sortPriority=78}) {name=isCharAccent, nodeColor=<rgba 0 0.9253 1 1>}
  Saturate(None=Add(A=@n92) {constB=0.25})
  ScalarParameter() {group=Wrap_DinoHunter, parameterName=DinoH_Metallic, sortPriority=0}
 ]}
