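# Material function for the animated lava lamp cosmetic wrap, with tuning for weapons, vehicles, mechs and characters.
# Projected pre-skinned UVs, distorted by scrolling warp textures, drive liquid masks and two randomized sine-animated bubble layers.
# The masks index curve-atlas colors over a hue-shiftable background, adding fake sun specular and time-of-day-scaled emissive.
# Customization Mask A applies the lava, Mask B the accent material, behind the Wrap_LavaLamp static switch.
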MaterialFunction /Game/Weapons/WeaponSkins/Wraps/MaterialFunctions/MF_Wrap_LavaLamp
graph {nodes [
  n1 = GetMaterialAttributes(None=FunctionInput() {bUsePreviewValueAsDefault=true, inputName=Material, inputType=FunctionInput_MaterialAttributes, sortPriority=-20}) {attributeGetTypes [
    #1
    #2
    #3
    #4
    #5
   ]}
  n2 = VectorParameter() {
   group=Wrap_LavaLamp_Accent
   parameterName=LavaL_2nd_BaseColor_A
   sortPriority=1
   channelNames {a="Use 3=Colors   0=on", b="", g="", r=""}
   defaultValue=<rgba 0 0 0 1>
  }
  n3 = VectorParameter() {group=Wrap_LavaLamp_Accent, parameterName=LavaL_2nd_BaseColor_B, sortPriority=1, defaultValue=<rgba 0 0 0 1>}
  n4 = VectorParameter() {group=Wrap_LavaLamp_Accent, parameterName=LavaL_2nd_BaseColor_C, sortPriority=1, defaultValue=<rgba 0 0 0 1>}
  n5 = MaterialFunctionCall(A=@n2, B=@n3.RGB, C=@n4.RGB, Alpha=ComponentMask(None=NamedRerouteUsage()) {r=true}) {materialFunction=&3ColorBlend}
  n6 = FunctionInput(Preview=TextureSample() {samplerType=SAMPLERTYPE_Masks, texture=/Game/Weapons/FORT_Rifles/Textures/T_Scar_CM}.G) {bUsePreviewValueAsDefault=true, inputName=Curavture, inputType=FunctionInput_Scalar, sortPriority=24}
  n7 = Saturate(None=Multiply(A=LinearInterpolate(A=@n5.Result, B=@n2, Alpha=@n2), B=Add(A=@n6) {constB=0.5}))
  n8 = ScalarParameter() {group=Wrap_LavaLamp_Accent, parameterName=LavaL_2nd_Roghness, sortPriority=1}
  n9 = ScalarParameter() {defaultValue=0.5, group=Wrap_LavaLamp_Accent, parameterName=LavaL_2nd_Specular, sortPriority=1}
  n10 = Time()
  n11 = ScalarParameter() {group=Wrap_LavaLamp, parameterName=LavaL_HueShift_C}
  n12 = VectorParameter() {
   group=Wrap_LavaLamp_BackGround
   parameterName=LavaL_BackGrond_Alt_Axis_Value
   channelNames {a="Use AltAxies Msk  0=on", b="", g="", r=Value used}
   defaultValue=<rgba 0 0 0 1>
  }
  n13 = ScalarParameter() {defaultValue=3, group=Wrap_LavaLamp_BackGround, parameterName=LavaL_BackGround_TileRate}
  n14 = ScalarParameter() {defaultValue=3, group=Wrap_LavaLamp_BackGround, parameterName=LavaL_BackGround_TileRate_Character}
  n15 = NamedRerouteUsage()
  n16 = VertexInterpolator(VS=PreSkinnedPosition())
  n17 = PreSkinnedLocalBounds()
  n18 = FunctionInput(Preview=Divide(A=Subtract(A=@n16.PS, B=@n17.Min), B=Subtract(A=@n17.Max, B=@n17.Max))) {bUsePreviewValueAsDefault=true, inputName=Calcated_Bounds, sortPriority=11}
  n19 = FunctionInput(Preview=VertexInterpolator(VS=PreSkinnedNormal()).PS) {bUsePreviewValueAsDefault=true, inputName=Pre-Skinned Local Normal, sortPriority=10}
  n20 = FunctionInput(Preview=Constant4Vector().RGBA) {inputName=IsState, inputType=FunctionInput_Vector4, sortPriority=5}
  n21 = MaterialFunctionCall(Color=@n20) {materialFunction=&SplitComponents}
  n22 = MaterialFunctionCall!(
   For_Safe_Zones
   Normal_Map
   Caculated_Bounds=@n18
   Pre_Skinned_Normals=@n19
   Is Vehicle=@n21.R
   Is VehicleSide=@n21.R
   Is Mech=@n21.R
   Vehicle_Rotation
   UV_Rotation Bypass
   Mech_Rotation
   Uvs_Offset
   Offset-after rotation
   Uvs_TileRate
  ) {materialFunction=&/MF_Projection_Side_Normal}
  n23 = Constant() {r=0.5}
  n24 = ScalarParameter() {group=Wrap_LavaLamp, parameterName=LavaL_Rotation, sortPriority=10}
  n25 = MaterialFunctionCall(UVs=@n22.UVs, Rotation Center=@n23, "Rotation Angle (0-1)"=@n24) {materialFunction=&CustomRotator}
  n26 = LinearInterpolate(
   A=ScalarParameter() {defaultValue=3, group=Wrap_LavaLamp, parameterName=LavaL_Liquid_Warp_TileRate_Multp}
   B=ScalarParameter() {defaultValue=3, group=Wrap_LavaLamp, parameterName=LavaL_Liquid_Warp_TileRate_Multp_Character}
   Alpha=NamedRerouteUsage()
  )
  n27 = Multiply(A=ScalarParameter() {defaultValue=0.5, group=Wrap_LavaLamp, parameterName=LavaL_Liquid_Warp_TileRate_A}, B=@n26)
  n28 = MaterialFunctionCall(UVs=@n22.UVs No Offsets/Tiling, Rotation Center=@n23, "Rotation Angle (0-1)"=@n24) {materialFunction=&CustomRotator}
  n29 = Time()
  n30 = ComponentMask(None=VectorParameter() {group=Wrap_LavaLamp, parameterName=LavaL_Liquid_Warp_Speed_A, defaultValue=<rgba 0.01 0.1 0 1>}.RGB) {g=true, r=true}
  n31 = NamedRerouteUsage()
  n32 = TextureSample(UVs=Add(A=Multiply(A=@n28.Rotated Values, B=@n27), B=Multiply(A=@n29, B=@n30)), Tex=@n31)
  n33 = ChannelMaskParameter(None=@n32.RGBA) {group=Wrap_LavaLamp, maskChannel=Alpha, parameterName=LavaL__Mask__Warp_Channel, defaultValue=<rgba 0 0 0 1>}
  n34 = Multiply(
   A=@n28.Rotated Values
   B=Multiply(A=ScalarParameter() {defaultValue=1, group=Wrap_LavaLamp, parameterName=LavaL_Liquid_Warp_TileRate_B}, B=@n26)
  )
  n35 = Multiply(
   A=@n29
   B=ComponentMask(None=VectorParameter() {group=Wrap_LavaLamp, parameterName=LavaL_Liquid_Warp_Speed_B, defaultValue=<rgba 0.01 0.1 0 1>}.RGB) {g=true, r=true}
  )
  n36 = ChannelMaskParameter(None=TextureSample(UVs=Add(A=@n34, B=@n35), Tex=@n31).RGBA) {group=Wrap_LavaLamp, maskChannel=Alpha, parameterName=LavaL__Mask__Warp_Channel, defaultValue=<rgba 0 0 0 1>}
  n37 = Add(
   A=Multiply(A=Add(A=@n33.R) {constB=-0.5}, B=ScalarParameter() {defaultValue=0.2, group=Wrap_LavaLamp, parameterName=LavaL_Liquid_Warp_Amount_A})
   B=Multiply(A=Add(A=@n36.R) {constB=-0.5}, B=ScalarParameter() {defaultValue=0.1, group=Wrap_LavaLamp, parameterName=LavaL_Liquid_Warp_Amount_B})
  )
  n38 = Add(
   A=Multiply(A=@n25.Rotated Values, B=LinearInterpolate(A=@n13, B=@n14, Alpha=@n15)) {constB=3}
   B=Multiply(A=@n37, B=ScalarParameter() {defaultValue=0.5, group=Wrap_LavaLamp_BackGround, parameterName=LavaL_BackGround_Warp_Amount}) {constB=0.5}
  )
  n39 = Multiply(
   A=Time()
   B=ComponentMask(None=VectorParameter() {group=Wrap_LavaLamp_BackGround, parameterName=LavaL_BackGrond_Speed, defaultValue=<rgba 0 0 0 1>}.RGB) {g=true, r=true}
  )
  n40 = TextureObjectParameter() {group=Wrap_LavaLamp, parameterName=LavaL__Roghness__Tex, texture=&T_BlacksRGB}
  n41 = ChannelMaskParameter(None=TextureSample(UVs=Add(A=@n38, B=@n39), Tex=@n40) {samplerType=SAMPLERTYPE_LinearColor}.RGBA) {group=Wrap_LavaLamp, maskChannel=Alpha, parameterName=LavaL__Roghness__Background_Channel, defaultValue=<rgba 0 0 0 1>}
  n42 = ComponentMask(None=@n25.Rotated Values) {r=true}
  n43 = Multiply(A=@n42, B=OneMinus(None=@n42))
  n44 = Saturate(None=Add(A=Multiply(A=@n43) {constB=3}) {constB=0.25})
  n45 = NamedRerouteUsage()
  n46 = LinearInterpolate(
   A=VectorParameter() {group=Wrap_LavaLamp_BackGround, parameterName=LavaL_BackGrond_Color_A, defaultValue=<rgba 0 0 0 0.1>}.RGB
   B=VectorParameter() {group=Wrap_LavaLamp_BackGround, parameterName=LavaL_BackGrond_Color_B, defaultValue=<rgba 0 0 0 0.1>}.RGB
   Alpha=LinearInterpolate(A=@n12.Value used, B=Multiply(A=@n41.R, B=@n44), Alpha=Max(A=@n45, B=@n12."Use AltAxies Msk  0=on"))
  )
  n47 = ScalarParameter() {defaultValue=0.5, group=Wrap_LavaLamp, parameterName=LavaL_Liquid_Spread_WEapon}
  n48 = MaterialFunctionCall(Color=@n20) {materialFunction=&SplitComponents}
  n49 = LinearInterpolate(
   A=ScalarParameter() {defaultValue=0.5, group=Wrap_LavaLamp, parameterName=LavaL_Liquid_Spread_Vehicle}
   B=ScalarParameter() {defaultValue=0.5, group=Wrap_LavaLamp, parameterName=LavaL_Liquid_Spread_Vehicle_Side}
   Alpha=@n48.G
  )
  n50 = AppendVector(
   A=Constant()
   B=LinearInterpolate(
    A=LinearInterpolate(A=@n47, B=@n49, Alpha=@n48.R)
    B=ScalarParameter() {defaultValue=0.5, group=Wrap_LavaLamp, parameterName=LavaL_Liquid_Spread_Mech}
    Alpha=@n48.B
   )
  )
  n51 = VectorParameter() {group=Wrap_LavaLamp, parameterName=LavaL_Liquid_Offset_Vehicle, defaultValue=<rgba 0 -1 0 1>}
  n52 = VectorParameter() {group=Wrap_LavaLamp, parameterName=LavaL_Liquid_Offset_VehicleSide}
  n53 = MaterialFunctionCall(Color=@n20) {materialFunction=&SplitComponents}
  n54 = LinearInterpolate(
   A=LinearInterpolate(
    A=VectorParameter() {group=Wrap_LavaLamp, parameterName=LavaL_Liquid_Offset_Weapon}.RGB
    B=LinearInterpolate(A=@n51.RGB, B=@n52.RGB, Alpha=@n53.G)
    Alpha=@n53.R
   )
   B=VectorParameter() {group=Wrap_LavaLamp, parameterName=LavaL_Liquid_Offset_Mech}.RGB
   Alpha=@n53.B
  )
  n55 = Add(A=@n6) {constB=-0.5}
  n56 = Multiply(A=@n55, B=ScalarParameter() {defaultValue=0.01, group=Wrap_LavaLamp, parameterName=LavaL_Liquid_Warp_Curavture_Amount})
  n57 = Add(A=Add(A=ComponentMask(None=@n54) {g=true, r=true}, B=@n28.Rotated Values), B=@n56)
  n58 = Add(B=Multiply(A=Constant2Vector() {g=-1, r=1}, B=@n57)) {constA=-1}
  n59 = LinearInterpolate(
   A=ScalarParameter() {defaultValue=2, group=Wrap_LavaLamp, parameterName=LavaL_Liquid_Thickness_Weapon}
   B=ScalarParameter() {defaultValue=2, group=Wrap_LavaLamp, parameterName=LavaL_Liquid_Thickness_Weapon_Pliant_Pickaxe}
   Alpha=NamedRerouteUsage()
  )
  n60 = MaterialFunctionCall(Color=@n20) {materialFunction=&SplitComponents}
  n61 = LinearInterpolate(
   A=ScalarParameter() {defaultValue=2, group=Wrap_LavaLamp, parameterName=LavaL_Liquid_Thickness_Vehicle}
   B=ScalarParameter() {defaultValue=2, group=Wrap_LavaLamp, parameterName=LavaL_Liquid_Thickness_Vehicle_Side}
   Alpha=@n60.G
  )
  n62 = AppendVector(
   A=Constant() {r=1}
   B=LinearInterpolate(
    A=LinearInterpolate(A=@n59, B=@n61, Alpha=@n60.R)
    B=ScalarParameter() {defaultValue=2, group=Wrap_LavaLamp, parameterName=LavaL_Liquid_Thickness_Mech}
    Alpha=@n60.B
   )
  )
  n63 = Add(A=Multiply(A=Add(A=@n50, B=@n58), B=@n62), B=@n37)
  n64 = NamedRerouteUsage()
  n65 = Multiply(A=Add(A=@n50, B=@n57), B=@n62) {constA=1}
  n66 = ChannelMaskParameter(None=TextureSample(UVs=Add(A=@n65, B=@n37), Tex=@n64).RGBA) {group=Wrap_LavaLamp, parameterName=LavaL__Mask_Tex_Cl_Liquid_Channel}
  n67 = Add(
   A=ChannelMaskParameter(None=TextureSample(UVs=@n63, Tex=@n64).RGBA) {group=Wrap_LavaLamp, parameterName=LavaL__Mask_Tex_Cl_Liquid_Channel}.R
   B=@n66.R
  )
  n68 = OneMinus(None=@n67)
  n69 = LinearInterpolate(
   A=ScalarParameter() {defaultValue=20, group=Wrap_LavaLamp, parameterName=LavaL_Mask_Bubble_to_liquid_Clamp}
   B=ScalarParameter() {defaultValue=10, group=Wrap_LavaLamp, parameterName=LavaL_Mask_Bubble_to_liquid_Clamp_Character}
   Alpha=NamedRerouteUsage()
  )
  n70 = Multiply(
   A=@n28.Rotated Values
   B=Multiply(A=@n26, B=ScalarParameter() {defaultValue=1, group=Wrap_LavaLamp_Bubbles, parameterName=LavaL_Bubbles_Warp_TileRate_A})
  ) {constB=0.5}
  n71 = Multiply(
   A=Time()
   B=ComponentMask(None=VectorParameter() {group=Wrap_LavaLamp_Bubbles, parameterName=LavaL_Bubbles_Warp_Speed_A, defaultValue=<rgba 0.1 0 0 1>}.RGB) {g=true, r=true}
  )
  n72 = ChannelMaskParameter(None=TextureSample(UVs=Add(A=@n70, B=@n71), Tex=NamedRerouteUsage()).RGBA) {group=Wrap_LavaLamp, maskChannel=Alpha, parameterName=LavaL__Mask__Warp_Channel, defaultValue=<rgba 0 0 0 1>}
  n73 = Multiply(
   A=Add(A=@n72.R, B=ScalarParameter() {defaultValue=-0.5, group=Wrap_LavaLamp_Bubbles, parameterName=LavaL_Bubbles_Warp_Offset}) {constB=-0.5}
   B=ScalarParameter() {defaultValue=0.25, group=Wrap_LavaLamp_Bubbles, parameterName=LavaL_Bubbles_Warp_amount_A}
  )
  n74 = Constant2Vector() {g=-0.4}
  n75 = LinearInterpolate(
   A=ScalarParameter() {defaultValue=5, group=Wrap_LavaLamp_Bubbles, parameterName=LavaL_Bubbles_Number_A}
   B=ScalarParameter() {defaultValue=5, group=Wrap_LavaLamp_Bubbles, parameterName=LavaL_Bubbles_Number_A_Character}
   Alpha=NamedRerouteUsage()
  )
  n76 = Constant() {r=1}
  n77 = VectorParameter() {group=Wrap_LavaLamp_Bubbles, parameterName=LavaL_Bubbles_Offset_Vehicle}
  n78 = VectorParameter() {group=Wrap_LavaLamp_Bubbles, parameterName=LavaL_Bubbles_Offset_VehicleSide}
  n79 = MaterialFunctionCall(Color=@n20) {materialFunction=&SplitComponents}
  n80 = LinearInterpolate(
   A=LinearInterpolate(
    A=VectorParameter() {group=Wrap_LavaLamp_Bubbles, parameterName=LavaL_Bubbles_Offset_Weapon}.RGB
    B=LinearInterpolate(A=@n77.RGB, B=@n78.RGB, Alpha=@n79.G)
    Alpha=@n79.R
   )
   B=VectorParameter() {group=Wrap_LavaLamp_Bubbles, parameterName=LavaL_Bubbles_Offset_Mech}.RGB
   Alpha=@n79.B
  )
  n81 = Add(A=Add(A=@n28.Rotated Values, B=ComponentMask(None=@n80) {g=true, r=true}), B=@n56)
  n82 = Multiply(A=@n81, B=AppendVector(A=@n75, B=@n76))
  n83 = Frac(None=@n82)
  n84 = Add(
   A=Floor(None=@n82)
   B=ComponentMask(None=VectorParameter() {group=Wrap_LavaLamp_Bubbles, parameterName=LavaL_Bubbles_Random_seed, defaultValue=<rgba 3 6 0 0>}.RGB) {g=true, r=true}
  )
  n85 = VectorParameter() {group=Wrap_LavaLamp_Bubbles, parameterName=LavaL_Bubbles_Speed_A, defaultValue=<rgba 0 0.051 0 1>}
  n86 = Add(
   A=AppendVector(
    A=Constant()
    B=MaterialFunctionCall!(2 Vector=@n84, RandomSeed) {materialFunction=&/MF_Random_2}.utput
   )
   B=Multiply(A=Time(), B=ComponentMask(None=@n85.RGB) {g=true, r=true})
  )
  n87 = MaterialFunctionCall(Color=@n20) {materialFunction=&SplitComponents}
  n88 = LinearInterpolate(
   A=ScalarParameter() {defaultValue=1, group=Wrap_LavaLamp_Bubbles, parameterName=LavaL_Bubbles_Sine_Strenght_A_Weapon}
   B=LinearInterpolate(
    A=ScalarParameter() {defaultValue=1, group=Wrap_LavaLamp_Bubbles, parameterName=LavaL_Bubbles_Sine_Strenght_A_Vehicle}
    B=ScalarParameter() {defaultValue=1, group=Wrap_LavaLamp_Bubbles, parameterName=LavaL_Bubbles_Sine_Strenght_A_Vehicle_Side}
    Alpha=@n87.G
   )
   Alpha=@n87.R
  )
  n89 = ScalarParameter() {defaultValue=1, group=Wrap_LavaLamp_Bubbles, parameterName=LavaL_Bubbles_Sine_Strenght_A_Mech}
  n90 = Add(
   A=Multiply(A=Add(A=@n74, B=@n83) {constA=1}, B=AppendVector(A=@n76, B=@n75))
   B=Multiply(A=Sine(None=@n86), B=LinearInterpolate(A=@n88, B=@n89, Alpha=@n87.B))
  )
  n91 = NamedRerouteUsage()
  n92 = MaterialFunctionCall!(
   UVs=@n28.Rotated Values
   Texture Size=ScalarParameter() {defaultValue=256, group=Wrap_LavaLamp_Bubbles, parameterName=LavaL_Bubbles_Miplevel _Size}
  ) {materialFunction=/Engine/Functions/Engine_MaterialFunctions01/Texturing/ComputeMipLevel}
  n93 = ChannelMaskParameter(None=TextureSample!(UVs=Add(A=@n73, B=@n90), Tex=@n91, Level=@n92.MipLevel, Apply View MipBias) {mipValueMode=TMVM_MipLevel}.RGBA) {group=Wrap_LavaLamp, maskChannel=Blue, parameterName=LavaL__Mask_Tex_Cl_Bubble_Channel, defaultValue=<rgba 0 0 1 0>}
  n94 = Constant2Vector() {g=-0.4}
  n95 = LinearInterpolate(
   A=ScalarParameter() {defaultValue=4, group=Wrap_LavaLamp_Bubbles, parameterName=LavaL_Bubbles_Number_B}
   B=ScalarParameter() {defaultValue=4, group=Wrap_LavaLamp_Bubbles, parameterName=LavaL_Bubbles_Number_B_Character}
   Alpha=NamedRerouteUsage()
  )
  n96 = Constant() {r=1}
  n97 = Multiply(A=Add(A=@n81, B=Constant2Vector() {r=0.12}), B=AppendVector(A=@n95, B=@n96))
  n98 = Frac(None=@n97)
  n99 = Add(
   A=Floor(None=@n97)
   B=ComponentMask(None=VectorParameter() {group=Wrap_LavaLamp_Bubbles, parameterName=LavaL_Bubbles_Random_seed, defaultValue=<rgba 5 6 0 0>}.RGB) {g=true, r=true}
  )
  n100 = VectorParameter() {group=Wrap_LavaLamp_Bubbles, parameterName=LavaL_Bubbles_Speed_B, defaultValue=<rgba 0 0.023 0 1>}
  n101 = Add(
   A=AppendVector(A=Constant(), B=MaterialFunctionCall(2 Vector=@n99) {materialFunction=&MF_Random_2}.utput)
   B=Multiply(A=Time(), B=ComponentMask(None=@n100.RGB) {g=true, r=true})
  )
  n102 = MaterialFunctionCall(Color=@n20) {materialFunction=&SplitComponents}
  n103 = LinearInterpolate(
   A=ScalarParameter() {defaultValue=1, group=Wrap_LavaLamp_Bubbles, parameterName=LavaL_Bubbles_Sine_Strenght_B_Weapon}
   B=LinearInterpolate(
    A=ScalarParameter() {defaultValue=1, group=Wrap_LavaLamp_Bubbles, parameterName=LavaL_Bubbles_Sine_Strenght_B_Vehicle}
    B=ScalarParameter() {defaultValue=1, group=Wrap_LavaLamp_Bubbles, parameterName=LavaL_Bubbles_Sine_Strenght_B_Vehicle_Side}
    Alpha=@n102.G
   )
   Alpha=@n102.R
  )
  n104 = ScalarParameter() {defaultValue=1, group=Wrap_LavaLamp_Bubbles, parameterName=LavaL_Bubbles_Sine_Strenght_B_Mech}
  n105 = Add(
   A=Multiply(A=Add(A=@n94, B=@n98) {constA=1}, B=AppendVector(A=@n96, B=@n95))
   B=Multiply(A=Sine(None=@n101), B=LinearInterpolate(A=@n103, B=@n104, Alpha=@n102.B))
  )
  n106 = ChannelMaskParameter(None=TextureSample(UVs=Add(A=@n73, B=@n105), Tex=@n91, Level=@n92.MipLevel) {mipValueMode=TMVM_MipLevel}.RGBA) {group=Wrap_LavaLamp, maskChannel=Blue, parameterName=LavaL__Mask_Tex_Cl_Bubble_Channel, defaultValue=<rgba 0 0 1 0>}
  n107 = Multiply(A=Saturate(None=Power(Base=@n68, Exp=@n69) {constExponent=20}), B=Add(A=@n93.R, B=@n106.R))
  n108 = Add(A=@n67, B=@n107)
  n109 = CurveAtlasRowParameter(CurveTime=@n108) {
   atlas=/Game/Weapons/WeaponSkins/Wraps/Textures/CruveAtlas_Wraps_01
   curve=/Game/Weapons/WeaponSkins/Wraps/Textures/LavaLamp/CC_Liquid_MAsk
   defaultValue=53
   group=Wrap_LavaLamp
   parameterName=LavaL_Liquid_ColorCuve_Mask
  }
  n110 = MaterialFunctionCall!(
   Index=ScalarParameter() {defaultValue=53, group=Wrap_LavaLamp, parameterName="LavaL_Liquid_ColorCuve_Mask  Index"}
   CCinput=@n109
   CCinput_Alpha=@n109
   UV=@n108
  ) {materialFunction=&/MF_Global_CC_Tex}
  n111 = ComponentMask(None=@n110.3V) {r=true}
  n112 = Multiply(A=@n10, B=ScalarParameter() {group=Wrap_LavaLamp, parameterName=LavaL_HueShift_A})
  n113 = CurveAtlasRowParameter(CurveTime=@n108) {
   atlas=&CruveAtlas_Wraps_01
   curve=/Game/Weapons/WeaponSkins/Wraps/Textures/LavaLamp/CC_Liquid_Color
   defaultValue=52
   group=Wrap_LavaLamp
   parameterName=LavaL_Liquid_ColorCuve
  }
  n114 = MaterialFunctionCall(
   Index=ScalarParameter() {defaultValue=52, group=Wrap_LavaLamp, parameterName="LavaL_Liquid_ColorCuve  Index"}
   CCinput=@n113
   CCinput_Alpha=@n113
   UV=@n108
  ) {materialFunction=&MF_Global_CC_Tex}
  n115 = Multiply(A=@n10, B=ScalarParameter() {group=Wrap_LavaLamp, parameterName=LavaL_HueShift_B})
  n116 = CurveAtlasRowParameter(CurveTime=@n108) {
   atlas=&CruveAtlas_Wraps_01
   curve=/Game/Weapons/WeaponSkins/Wraps/Textures/LavaLamp/CC_Liquid_ColorB
   defaultValue=55
   group=Wrap_LavaLamp
   parameterName=LavaL_Liquid_ColorCuve_B
  }
  n117 = ScalarParameter() {defaultValue=1, group=Wrap_LavaLamp, parameterName=LavaL_Mask_Liquid_to_Bubbles_Clamp}
  n118 = Multiply(A=Add(A=@n107, B=@n66.R), B=Saturate(None=Power(Base=@n66.R, Exp=@n117))) {constA=1}
  n119 = CurveAtlasRowParameter(CurveTime=@n118) {
   atlas=&CruveAtlas_Wraps_01
   curve=/Game/Weapons/WeaponSkins/Wraps/Textures/LavaLamp/CC_Liquid_ColorC
   defaultValue=56
   group=Wrap_LavaLamp
   parameterName=LavaL_Liquid_ColorCuve_C
  }
  n120 = Add(
   A=MaterialFunctionCall(Index=ScalarParameter() {defaultValue=55, group=Wrap_LavaLamp, parameterName="LavaL_Liquid_ColorCuve_B  Index"}, CCinput=@n116, UV=@n108) {materialFunction=&MF_Global_CC_Tex}.3V
   B=MaterialFunctionCall(Index=ScalarParameter() {defaultValue=56, group=Wrap_LavaLamp, parameterName="LavaL_Liquid_ColorCuve_C  Index"}, CCinput=@n119, UV=@n118) {materialFunction=&MF_Global_CC_Tex}.3V
  )
  n121 = LinearInterpolate(
   A=MaterialFunctionCall(Hue Shift Percentage=Multiply(A=@n10, B=@n11), Texture=Multiply(A=@n46, B=@n111)) {materialFunction=&HueShift}.Result
   B=Add(
    A=MaterialFunctionCall(Hue Shift Percentage=@n112, Texture=@n114.3V) {materialFunction=&HueShift}.Result
    B=MaterialFunctionCall(Hue Shift Percentage=@n115, Texture=@n120) {materialFunction=&HueShift}.Result
   )
   Alpha=@n114.Alpha
  )
  n122 = Saturate(None=Multiply(A=@n55, B=ScalarParameter() {defaultValue=2, group=Wrap_LavaLamp_Base, parameterName=LavaL_Roghness_Curvature Strength}) {constB=2})
  n123 = LinearInterpolate(
   A=@n121
   B=Multiply(A=ScalarParameter() {defaultValue=1, group=Wrap_LavaLamp, parameterName=LavaL_Edge_Brightness, sortPriority=0}, B=@n121)
   Alpha=@n122
  )
  n124 = CollectionParameter() {
   collection=&FortniteMaterialParameters
   parameterName=SunlightDirectionalVector
  }
  n125 = ComponentMask(None=@n124) {b=true, g=true, r=true}
  n126 = MaterialFunctionCall!(Value=@n110.Alpha) {materialFunction=/Engine/Functions/Engine_MaterialFunctions03/Procedurals/PreparePerturbNormalHQ}
  n127 = MaterialFunctionCall!(
   Bump one pixel right=@n126.Value one pixel right
   Bump center=@n126.Value one pixel right
   Bump one pixel down=@n126.Value one pixel right
   WS Normal
  ) {materialFunction=/Engine/Functions/Engine_MaterialFunctions03/Procedurals/PerturbNormalHQ}
  n128 = Multiply(A=DotProduct(A=@n125, B=@n127.WS Normal)) {constB=2}
  n129 = Normalize(VectorInput=Subtract(A=Multiply(A=@n128, B=@n127.WS Normal), B=@n125))
  n130 = Multiply(A=MaterialFunctionCall() {materialFunction=&CameraDirectionVector}.Result) {constB=-1}
  n131 = Multiply(
   A=Power(
    Base=DotProduct(A=@n129, B=@n130)
    Exp=ScalarParameter() {defaultValue=50, group=Wrap_LavaLamp_TOD, parameterName=LavaL_FakeLighting_Mask_range, sortPriority=100}
   )
   B=ScalarParameter() {defaultValue=0.3, group=Wrap_LavaLamp_TOD, parameterName=LavaL_FakeLighting_Specular_Scale, sortPriority=100}
  ) {constB=0.4}
  n132 = Multiply(
   A=Multiply(A=@n114.Alpha, B=Saturate(None=@n131))
   B=VectorParameter() {group=Wrap_LavaLamp_TOD, parameterName=LavaL_FakeLighting_Color, defaultValue=<rgba 1 1 1 1>}.RGB
  )
  n133 = Multiply(
   A=Multiply(
    A=Add(A=@n123, B=@n132)
    B=ScalarParameter() {defaultValue=1, group=Wrap_LavaLamp_Base, parameterName=LavaL_Emissive_Brightness, sortPriority=0}
   )
   B=MaterialFunctionCall(
    Value_Morning=ScalarParameter() {defaultValue=0.5, group=Wrap_LavaLamp_TOD, parameterName=LavaLTOD_Morning, sortPriority=1}
    Value_Day=ScalarParameter() {defaultValue=1, group=Wrap_LavaLamp_TOD, parameterName=LavaLTOD_Day, sortPriority=1}
    Value_Evening=ScalarParameter() {defaultValue=0.5, group=Wrap_LavaLamp_TOD, parameterName=LavaLTOD_Evening, sortPriority=1}
    Value_Night=ScalarParameter() {defaultValue=0.25, group=Wrap_LavaLamp_TOD, parameterName=LavaLTOD_Night, sortPriority=1}
   ) {materialFunction=&Scalar_Time_of_Day}.Result
  )
  n134 = ScalarParameter() {group=Wrap_LavaLamp_Accent, parameterName=LavaL_2nd_Metallic, sortPriority=1}
  n135 = SetMaterialAttributes!(MaterialAttributes=@n1.MaterialAttributes, BaseColor=@n7, Roughness=@n8, Specular=@n9, EmissiveColor=@n133, Metallic=@n134) {attributeSetTypes [
    #1
    #2
    #3
    #5
    #4
   ]}
  n136 = TextureSample!(UVs, Tex, Apply View MipBias) {samplerType=SAMPLERTYPE_Masks, texture=&T_Scar_CM}
  n137 = FunctionInput(Preview=Saturate(None=Add(A=@n136.R, B=@n136.R))) {bUsePreviewValueAsDefault=true, inputName=Customization Mask A, inputType=FunctionInput_Scalar, sortPriority=24}
  BlendMaterialAttributes(A=@n1.MaterialAttributes, B=@n135, Alpha=Saturate(None=@n137)) {vertexAttributeBlendType=UseA}
  n138 = Constant() {r=1}
  n139 = VectorParameter() {group=Wrap_LavaLamp_Base, parameterName=LavaL_Roghness_Altreranate_Axis_color, defaultValue=<rgba 0 0 0 1>}
  n140 = MaterialFunctionCall(Color=@n20) {materialFunction=&SplitComponents}
  n141 = MaterialFunctionCall!(
   IsCharacter
   IsStatsB
   Texture=@n40
   Rotation_Axis=Constant3Vector()
   Caculated_Bounds=Multiply(A=ScalarParameter() {defaultValue=1, group=Wrap_LavaLamp_Base, parameterName=LavaL_Roghness_TileRate}, B=@n18)
   Pre_Skinned_Normals=@n19
   Blend_Contrast=@n138
   Fill_alternate_Axis_Opacity=@n138
   "Fill_alternate_Axis_Color_A=Roughness"=AppendVector(A=@n139.RGB, B=@n139.RGB)
   Is_Vehicle=@n140.R
   Is_Mech=@n140.R
   Enable_Fill_Z_Weapon
   Fill_Z_Weapon_opacity
   Fill_Z_Weapon_Color
  ) {materialFunction=&/MF_TopsSideProjection_Base}
  ComponentMask(None=@n141.Combined) {r=true}
  n142 = VectorParameter() {group=Wrap_LavaLamp_Base, parameterName=LavaL_BaseColor, sortPriority=0, defaultValue=<rgba 0 0 0 1>}
  n143 = FunctionInput(Preview=@n136.B) {bUsePreviewValueAsDefault=true, inputName=Customization Mask B, inputType=FunctionInput_Scalar, sortPriority=24}
  n144 = ChannelMaskParameter(None=@n141.Combined) {group=Wrap_LavaLamp_Base, parameterName=LavaL_Roghness_Channel}
  n145 = LinearInterpolate(
   A=ScalarParameter() {group=Wrap_LavaLamp_Base, parameterName=LavaL_Roghness_Low}
   B=ScalarParameter() {defaultValue=1, group=Wrap_LavaLamp_Base, parameterName=LavaL_Roghness_High}
   Alpha=Saturate(None=Add(A=@n144.R, B=@n122))
  )
  n146 = ScalarParameter() {defaultValue=0.5, group=Wrap_LavaLamp_Base, parameterName=LavaL_Specular, sortPriority=0}
  n147 = ScalarParameter() {group=Wrap_LavaLamp_Base, parameterName=LavaL_Metallic, sortPriority=0}
  n148 = SetMaterialAttributes(
   MaterialAttributes=@n1.MaterialAttributes
   BaseColor=LinearInterpolate(A=LinearInterpolate(A=@n1.BaseColor, B=@n142.RGB, Alpha=@n137), B=@n7, Alpha=@n143)
   Roughness=LinearInterpolate(A=LinearInterpolate(A=@n1.Roughness, B=@n145, Alpha=@n137), B=@n8, Alpha=@n143)
   Specular=LinearInterpolate(A=LinearInterpolate(A=@n1.Specular, B=@n146, Alpha=@n137), B=@n9, Alpha=@n143)
   EmissiveColor=LinearInterpolate(A=LinearInterpolate(A=@n1.EmissiveColor, B=@n133, Alpha=@n137), Alpha=@n143) {constB=0}
   Metallic=LinearInterpolate(A=LinearInterpolate(A=@n1.Metallic, B=@n147, Alpha=@n137), B=@n134, Alpha=@n143)
  ) {attributeSetTypes [
    #1
    #2
    #3
    #5
    #4
   ]}
  FunctionOutput(None=StaticSwitchParameter(True=@n148, False=@n1.MaterialAttributes) {group=Customization_Wrap, parameterName=Wrap_LavaLamp})
  LinearInterpolate()
  n149 = ScalarParameter() {defaultValue=1, group=Wrap_LavaLamp, parameterName=LavaL_Liquid_Warp_TileRate_Multp_Vehicle}
  n150 = ScalarParameter() {defaultValue=1, group=Wrap_LavaLamp, parameterName=LavaL_Liquid_Warp_TileRate_Multp_Vehicle_Side}
  n151 = MaterialFunctionCall(Color=@n20) {materialFunction=&SplitComponents}
  LinearInterpolate(
   A=LinearInterpolate(
    A=ScalarParameter() {defaultValue=1, group=Wrap_LavaLamp, parameterName=LavaL_Liquid_Warp_TileRate_Multp_Weapon}
    B=LinearInterpolate(A=@n149, B=@n150, Alpha=@n151.G)
    Alpha=@n151.R
   )
   B=ScalarParameter() {defaultValue=1, group=Wrap_LavaLamp, parameterName=LavaL_Liquid_Warp_TileRate_Multp_Mech}
   Alpha=@n151.B
  )
  Multiply(A=@n106.R)
  Multiply(A=@n93.R)
  Multiply(A=@n32.A) {constB=0.5}
  Multiply(A=@n127.WS Normal)
  NamedRerouteDeclaration(None=MaterialFunctionCall(Float4=@n20) {materialFunction=&BreakOutFloat4Components}.A) {name=IsCharacter, nodeColor=<rgba 0.9909 1 0 1>}
  n152 = FunctionInput() {bUsePreviewValueAsDefault=true, desc="R=Backpack,G=Pickaxe", inputName=IsStatsB, inputType=FunctionInput_Vector2, sortPriority=5}
  NamedRerouteDeclaration(None=ComponentMask(None=@n152) {g=true}) {name=IsPickaxe, nodeColor=<rgba 0 0.6484 1 1>}
  NamedRerouteDeclaration(None=TextureObjectParameter() {group=Wrap_LavaLamp, parameterName=LavaL__Mask_Tex_Clamp, texture=&T_BlacksRGB}) {name=Tex_Clamp, nodeColor=<rgba 0 0.4609 1 1>}
  n153 = MaterialFunctionCall(Color=Abs(None=NamedRerouteUsage())) {materialFunction=&SplitComponents}
  n154 = MaterialFunctionCall(Float4=NamedRerouteUsage()) {materialFunction=&BreakOutFloat4Components}
  n155 = LinearInterpolate(A=@n153.R, B=LinearInterpolate(A=@n153.B, B=@n153.B, Alpha=@n154.G), Alpha=@n154.R)
  NamedRerouteDeclaration(None=MaterialFunctionCall(
    In=LinearInterpolate(A=@n155, B=@n153.G, Alpha=@n154.B)
    Contrast=ScalarParameter() {defaultValue=1, group=Wrap_LavaLamp_BackGround, parameterName=LavaL_Tri_Blend_Contrast, sortPriority=0}
   ) {materialFunction=&CheapContrast}.Result) {name=Tri_Mask, nodeColor=<rgba 1 0.4688 0 1>}
  NamedRerouteDeclaration(None=TextureObjectParameter() {group=Wrap_LavaLamp, parameterName=LavaL__Mask_Tex, texture=&T_BlacksRGB}) {name=Tex_Mask, nodeColor=<rgba 1 0 0.7031 1>}
  NamedRerouteDeclaration(None=MaterialFunctionCall!(
    Normal_Map
    Offset Rt
    Pivot
    Caculated_Bounds=@n18
    Pre_Skinned_Normals=@n19
    is_State=@n20
    UV_Rotation
    Uvs_Offset
    Uvs_TileRate
    Uvs_Offset_Pivot
    For_Safe_Zones
    Rotation_Comp
    Clamp Y
    Clamp X
    Clamp Padding_Comp
    No_Rotation=StaticBool() {value=true}
   ) {materialFunction=&/MF_Projection_Side_Normal_V2}.UVs) {name=Uvs, nodeColor=<rgba 0.4766 1 0 1>}
  NamedRerouteDeclaration(None=@n20) {name=Is_State, nodeColor=<rgba 0.6875 1 0 1>}
  NamedRerouteDeclaration(None=@n19) {name=SkinNormals, nodeColor=<rgba 0.9922 0 1 1>}
  Normalize(VectorInput=@n127.WS Normal)
  PixelNormalWS()
  n156 = ComponentMask(None=@n25.Rotated Values) {g=true}
  Saturate(None=Divide(A=Add(A=@n156) {constB=-0.25}) {constB=0.75})
  Saturate(None=Power(
    Base=ComponentMask(None=@n25.Rotated Values) {g=true}
    Exp=ScalarParameter() {defaultValue=3, group=Wrap_LavaLamp_BackGround, parameterName=LavaL_BackGround_Mask}
   ))
  VertexNormalWS()
 ]}
