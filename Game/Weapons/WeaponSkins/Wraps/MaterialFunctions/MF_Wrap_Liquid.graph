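# Wrap material function that fakes a glass container of bubbling liquid on weapons, vehicles, mechs, characters, backpacks and pickaxes, using state inputs to pick parameters.
# A height gradient, rippled by layered sines whose strength settles after a timestamp, drives curve-atlas color and masks.
# Virtual planes add the liquid's top surface, camera-facing glow and warped scrolling bubbles.
# The result feeds time-of-day-scaled emissive behind the Wrap_Liquid static switch, otherwise passing the input material through.
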MaterialFunction /Game/Weapons/WeaponSkins/Wraps/MaterialFunctions/MF_Wrap_Liquid
graph {nodes [
  n1 = FunctionInput(Preview=VertexInterpolator(VS=PreSkinnedPosition()).PS) {bUsePreviewValueAsDefault=true, inputName=Pre-Skinned Local Position, sortPriority=10}
  n2 = Multiply(A=ComponentMask(None=NamedRerouteUsage()) {b=true, g=true, r=true}) {constB=-1}
  n3 = Add(A=@n1, B=@n2) {constB=0}
  n4 = Transform(None=@n3) {transformSourceType=TRANSFORMSOURCE_Local}
  n5 = Add(
   A=Add(A=@n2, B=WorldPosition().XYZ)
   B=VectorParameter() {group=Wrap_Liquid, parameterName=Liquid_Surface__Z_Center_Offset_Pliant_Character}.RGB
  )
  n6 = NamedRerouteUsage()
  n7 = ComponentMask(None=Divide(
    A=LinearInterpolate(A=@n4, B=@n5, Alpha=@n6)
    B=ScalarParameter() {defaultValue=-100, group=Wrap_Liquid, parameterName=Liquid_Surface__Z_Falloff}
   ) {constB=100}) {b=true}
  Add(A=OneMinus(None=@n7)) {constB=-1}
  n8 = ObjectPositionWS()
  n9 = CameraPositionWS()
  n10 = Normalize(VectorInput=Multiply(A=Constant3Vector() {constant=<rgba 1 1 0 1>}, B=Subtract(A=@n8, B=@n9)))
  n11 = Constant3Vector() {constant=<rgba 0 0 1 1>}
  n12 = PreSkinnedLocalBounds()
  n13 = Subtract(A=@n12.Max, B=@n12.Max)
  n14 = MaterialFunctionCall(Color=@n13) {materialFunction=&SplitComponents}
  n15 = Max(A=@n14.R, B=@n14.R)
  n16 = MaterialFunctionCall(
   Plane Normal Axis=@n10
   Plane X Axis=Normalize(VectorInput=CrossProduct(A=@n10, B=@n11))
   Plane Center=@n8
   UV Scale=Multiply(
    A=Max(A=@n15, B=@n14.B)
    B=MaterialFunctionCall() {materialFunction=/Engine/Functions/Engine_MaterialFunctions02/WorldPositionOffset/ObjectScale}.Scale Y
   )
  ) {materialFunction=&VirtualPlaneCoordinates}
  n17 = VectorParameter() {group=Wrap_Liquid, parameterName=Liquid_BubbleWarp_Speed, defaultValue=<rgba 0.05 0.05 -0.1 0>}
  n18 = Time()
  n19 = Add(
   A=Multiply(A=@n16.UVs, B=ScalarParameter() {defaultValue=5, group=Wrap_Liquid, parameterName=Liquid_BubbleWarp_TileRate})
   B=ComponentMask(None=Multiply(A=@n17.RGB, B=@n18)) {b=true, g=true}
  )
  n20 = TextureObjectParameter() {group=Wrap_Liquid, parameterName=Liquid_BubbleWarp__Tex, texture=&T_BlacksRGB}
  n21 = Multiply(
   A=ChannelMaskParameter(None=TextureSample(UVs=@n19, Tex=@n20) {samplerType=SAMPLERTYPE_LinearGrayscale}.RGBA) {group=Wrap_Liquid, parameterName=Liquid_BubbleWarp_Mask_Channel}.R
   B=ScalarParameter() {defaultValue=0.02, group=Wrap_Liquid, parameterName=Liquid_BubbleWarp_Amount}
  )
  n22 = CameraPositionWS()
  n23 = Normalize(VectorInput=Multiply(A=Constant3Vector() {constant=<rgba 1 1 0 1>}, B=Subtract(A=@n8, B=@n22)))
  n24 = CrossProduct(A=@n23, B=Constant3Vector() {constant=<rgba 0 0 1 1>})
  n25 = PreSkinnedLocalBounds()
  n26 = Subtract(A=@n25.Max, B=@n25.Max)
  n27 = MaterialFunctionCall(Color=@n26) {materialFunction=&SplitComponents}
  n28 = Max(A=Max(A=@n27.R, B=@n27.R), B=@n27.B)
  n29 = MaterialFunctionCall() {materialFunction=&ObjectScale}
  Add(
   A=@n21
   B=MaterialFunctionCall(Plane Normal Axis=@n23, Plane X Axis=Normalize(VectorInput=@n24), Plane Center=@n8, UV Scale=Multiply(A=@n28, B=@n29.Scale Y)) {materialFunction=&VirtualPlaneCoordinates}.UVs
  )
  n30 = NamedRerouteUsage()
  n31 = ScalarParameter() {defaultValue=1, group=Wrap_Liquid, parameterName=Liquid_Surface_Distortion_TileRateBaseSize_Vehicle}
  n32 = ScalarParameter() {defaultValue=1, group=Wrap_Liquid, parameterName=Liquid_Surface_Distortion_TileRateBaseSize_VehicleSide}
  n33 = MaterialFunctionCall(Float4=NamedRerouteUsage()) {materialFunction=&BreakOutFloat4Components}
  n34 = LinearInterpolate(
   A=LinearInterpolate(
    A=ScalarParameter() {defaultValue=2, group=Wrap_Liquid, parameterName=Liquid_Surface_Distortion_TileRateBaseSize_Weapon}
    B=LinearInterpolate(A=@n31, B=@n32, Alpha=@n33.G)
    Alpha=@n33.R
   )
   B=LinearInterpolate(
    A=ScalarParameter() {defaultValue=1, group=Wrap_Liquid, parameterName=Liquid_Surface_Distortion_TileRateBaseSize_Mech}
    B=ScalarParameter() {defaultValue=1, group=Wrap_Liquid, parameterName=Liquid_Surface_Distortion_TileRateBaseSize_Character}
    Alpha=@n33.A
   )
   Alpha=@n33.B
  )
  n35 = Multiply(A=@n30, B=@n34)
  n36 = ComponentMask(None=@n35) {b=true, g=true}
  n37 = LinearInterpolate(A=@n36, B=ComponentMask(None=@n35) {b=true, r=true})
  n38 = Constant2Vector() {g=1, r=2}
  n39 = FunctionInput(Preview=Constant()) {bUsePreviewValueAsDefault=true, inputName=IsState, inputType=FunctionInput_Vector4}
  n40 = LinearInterpolate(
   A=@n36
   B=LinearInterpolate(
    A=Multiply(A=@n37, B=@n38)
    B=ComponentMask(None=@n35) {g=true, r=true}
    Alpha=ScalarParameter() {group=Wrap_Liquid, parameterName=Liquid_IsRoll}
   )
   Alpha=MaterialFunctionCall(Color=@n39) {materialFunction=&SplitComponents}.R
  )
  n41 = Multiply(A=@n40, B=ScalarParameter() {defaultValue=2, group=Wrap_Liquid, parameterName=Liquid_Surface_Distortion_TileRate A Master})
  n42 = ScalarParameter() {defaultValue=4.3, group=Wrap_Liquid, parameterName=Liquid_Surface_Distortion_TileRate A-A}
  n43 = Fmod(
   A=MaterialFunctionCall(Speed=ScalarParameter() {defaultValue=0.5, group=Wrap_Liquid, parameterName=Liquid_Surface_Distortion_Speed Max}) {materialFunction=&TimeWithSpeedVariable}.Time
   B=Constant() {r=1000}
  )
  n44 = Sine(None=Add(A=Multiply(A=@n41, B=@n42) {constB=14.4956}, B=@n43))
  n45 = Multiply(A=@n41, B=ScalarParameter() {defaultValue=2.13, group=Wrap_Liquid, parameterName=Liquid_Surface_Distortion_TileRate A-B}) {constB=-5.8402}
  n46 = Multiply(A=Sine(None=Add(A=@n45, B=@n43))) {constB=0.5}
  n47 = Sine(None=Add(A=@n41, B=@n43))
  n48 = MaterialFunctionCall(Float2=Add(A=Multiply(A=@n44) {constB=0.25}, B=Add(A=@n46, B=@n47))) {materialFunction=&BreakOutFloat2Components}
  n49 = VectorParameter() {group=Wrap_Liquid, parameterName=Liquid_Surface_Distortion_Vector Strenght, defaultValue=<rgba 1 1 1 0>}
  n50 = MaterialFunctionCall(Float2=@n49.RGB) {materialFunction=&BreakOutFloat2Components}
  n51 = Add(A=Multiply(A=@n48.R, B=@n50.R), B=Multiply(A=@n48.G, B=@n50.G))
  n52 = Multiply(A=@n40, B=ScalarParameter() {defaultValue=3, group=Wrap_Liquid, parameterName=Liquid_Surface_Distortion_TileRate B Master})
  n53 = ScalarParameter() {defaultValue=4.3, group=Wrap_Liquid, parameterName=Liquid_Surface_Distortion_TileRate B-A}
  n54 = Fmod(
   A=MaterialFunctionCall(Speed=ScalarParameter() {defaultValue=-1.03, group=Wrap_Liquid, parameterName=Liquid_Surface_Distortion_Speed B Max}) {materialFunction=&TimeWithSpeedVariable}.Time
   B=Constant() {r=1000}
  )
  n55 = Sine(None=Add(A=Multiply(A=@n52, B=@n53) {constB=14.4956}, B=@n54))
  n56 = Multiply(A=@n52, B=ScalarParameter() {defaultValue=2.13, group=Wrap_Liquid, parameterName=Liquid_Surface_Distortion_TileRate B-B}) {constB=-5.8402}
  n57 = Multiply(A=Sine(None=Add(A=@n56, B=@n54))) {constB=0.5}
  n58 = Sine(None=Add(A=@n52, B=@n54))
  n59 = MaterialFunctionCall(Float2=Add(A=Multiply(A=@n55) {constB=0.25}, B=Add(A=@n57, B=@n58))) {materialFunction=&BreakOutFloat2Components}
  n60 = MaterialFunctionCall(Float2=@n49.RGB) {materialFunction=&BreakOutFloat2Components}
  n61 = Multiply(A=@n59.R, B=@n60.R)
  n62 = Multiply(A=@n59.G, B=@n60.G)
  Add(A=@n51, B=Multiply(A=Add(A=@n61, B=@n62)) {constB=0.75})
  Constant3Vector() {constant=<rgba 0 0 1 0>}
  n63 = PreSkinnedLocalBounds()
  n64 = FunctionInput(Preview=Constant3Vector() {constant=<rgba 0 5 0 0>}) {bUsePreviewValueAsDefault=true, inputName=Object Bounds Min Adjustment, sortPriority=50}
  n65 = Add(A=@n63.Min, B=@n64) {constB=0}
  n66 = FunctionInput(Preview=Constant3Vector() {constant=<rgba -5 -25 -5 0>}) {bUsePreviewValueAsDefault=true, inputName=Object Bounds Max Adjustment, sortPriority=50}
  n67 = Add(A=@n63.Max, B=@n66) {constB=0}
  Divide(A=Subtract(A=@n3, B=@n65), B=Subtract(A=@n67, B=@n65))
  n68 = Subtract(A=@n1, B=@n65)
  n69 = Subtract(A=@n67, B=@n65)
  n70 = VectorParameter() {group=Wrap_Liquid, parameterName=Liquid_Surface__Z_Center_Offset_Pliant_Backpack, defaultValue=<rgba -0.5 -0.5 -0.5 1>}
  n71 = LinearInterpolate(
   A=VectorParameter() {group=Wrap_Liquid, parameterName=Liquid_Surface__Z_Center_Offset_Pliant_Character}.RGB
   B=@n70.RGB
   Alpha=NamedRerouteUsage()
  )
  n72 = Multiply(A=@n69, B=@n71) {constB=0.5}
  Divide(A=Transform(None=Add(A=@n68, B=@n72)) {transformSourceType=TRANSFORMSOURCE_Local}, B=@n69)
  Divide(A=@n30, B=@n34)
  n73 = TextureSample() {samplerType=SAMPLERTYPE_Normal, texture=/Game/Weapons/FORT_Rifles/T_SCAR_N}
  n74 = GetMaterialAttributes(None=FunctionInput(Preview=SetMaterialAttributes!(MaterialAttributes, Normal=@n73.RGB) {attributeSetTypes [#1]}) {inputName=Material, inputType=FunctionInput_MaterialAttributes, sortPriority=-20})
  n75 = VectorParameter() {group=Wrap_Liquid, parameterName=Liquid_Glass_Color, sortPriority=0, defaultValue=<rgba 0.0008 0.0019 0.0104 0>}
  n76 = ScalarParameter() {defaultValue=38, group=Wrap_Liquid, parameterName="Liquid_ColorCurve_Surface  Index"}
  n77 = VertexInterpolator(VS=PreSkinnedNormal())
  n78 = PreSkinnedLocalBounds()
  n79 = Add(A=@n78.Min, B=@n64)
  n80 = Add(
   A=Multiply(A=FunctionInput(Preview=@n77.PS) {bUsePreviewValueAsDefault=true, inputName=Pre-Skinned Local Normal, sortPriority=10}) {constB=0}
   B=@n79
  )
  n81 = Subtract(A=Add(A=@n78.Max, B=@n66), B=@n79)
  n82 = Multiply(A=@n81) {constB=0.5}
  n83 = MaterialFunctionCall(Float4=@n39) {materialFunction=&BreakOutFloat4Components}
  n84 = LinearInterpolate(A=TransformPosition(None=Add(A=@n80, B=@n82)) {transformType=TRANSFORMPOSSOURCE_World}, B=ObjectPositionWS(), Alpha=@n83.B)
  n85 = Normalize(VectorInput=Subtract(A=@n84, B=CameraPositionWS()))
  n86 = Constant3Vector() {constant=<rgba 0 0 1 1>}
  n87 = MaterialFunctionCall(Float3=@n81) {materialFunction=&BreakOutFloat3Components}
  n88 = Max(A=@n87.R, B=@n87.R)
  n89 = MaterialFunctionCall!(
   Plane Normal Axis=@n85
   Plane X Axis=Normalize(VectorInput=CrossProduct(A=@n85, B=@n86))
   Plane Center=@n84
   UV Scale=Multiply(
    A=MaterialFunctionCall() {materialFunction=&ObjectScale}.Scale X
    B=Max(A=@n88, B=@n87.B)
   )
  ) {materialFunction=/Game/Athena/Environments/Nexus/Materials/VirtualPlaneCoordinatesAbs}
  n90 = FunctionInput() {bUsePreviewValueAsDefault=true, inputName=Object-SafeZone-Virtual-CH-Switch, inputType=FunctionInput_Scalar, sortPriority=105}
  n91 = MaterialFunctionCall!(
   Camera _pos
   Axis
   Scale
   Scale_Multplyer
   Object-SafeZone-Virtual-Pivot=NamedRerouteUsage()
   Object-SafeZone-Virtual-CH-Switch=@n90
   SinglePlane Normal Axis
   SinglePlane X Axis
   PlaneCenter_distortion
  ) {materialFunction=&/MF_SafeZone_VirtualPlane}
  n92 = NamedRerouteUsage()
  n93 = TextureSample() {samplerType=SAMPLERTYPE_Masks, texture=/Game/Weapons/FORT_Rifles/Textures/T_Scar_CM}
  n94 = Add(A=FunctionInput(Preview=@n93.G) {bUsePreviewValueAsDefault=true, inputName=Curavture, inputType=FunctionInput_Scalar, sortPriority=24}) {constB=-0.5}
  n95 = Multiply(A=@n94, B=ScalarParameter() {defaultValue=-0.2, group=Wrap_Liquid, parameterName=Liquid_Outer_Glow_Curavatur Warp Amount})
  n96 = ScalarParameter() {defaultValue=1.252, group=Wrap_Liquid, parameterName=Liquid_Outer_Glow_Radus_Weapon}
  n97 = MaterialFunctionCall(Color=@n39) {materialFunction=&SplitComponents}
  n98 = LinearInterpolate(
   A=ScalarParameter() {defaultValue=1.252, group=Wrap_Liquid, parameterName=Liquid_Outer_Glow_Radus_Vehicle}
   B=ScalarParameter() {defaultValue=1.252, group=Wrap_Liquid, parameterName=Liquid_Outer_Glow_Radus_VehicleSide}
   Alpha=@n97.G
  )
  n99 = ScalarParameter() {defaultValue=-0.2095, group=Wrap_Liquid, parameterName=Liquid_Outer_Glow_Hardness_Weapon}
  n100 = MaterialFunctionCall(Color=@n39) {materialFunction=&SplitComponents}
  n101 = LinearInterpolate(
   A=ScalarParameter() {defaultValue=-0.2095, group=Wrap_Liquid, parameterName=Liquid_Outer_Glow_Hardness_Vehicle}
   B=ScalarParameter() {defaultValue=-0.2095, group=Wrap_Liquid, parameterName=Liquid_Outer_Glow_Hardness_VehicleSide}
   Alpha=@n100.G
  )
  n102 = SphereMask(
   A=Add(A=LinearInterpolate(A=@n89.UVs, B=@n91.Uvs, Alpha=@n92), B=@n95)
   B=Constant() {r=0.5}
   Radius=LinearInterpolate(
    A=LinearInterpolate(A=@n96, B=@n98, Alpha=@n97.R)
    B=ScalarParameter() {defaultValue=1.252, group=Wrap_Liquid, parameterName=Liquid_Outer_Glow_Radus_Mech}
    Alpha=@n97.B
   )
   Hardness=LinearInterpolate(
    A=LinearInterpolate(A=@n99, B=@n101, Alpha=@n100.R)
    B=ScalarParameter() {defaultValue=-0.2095, group=Wrap_Liquid, parameterName=Liquid_Outer_Glow_Hardness_Mech}
    Alpha=@n100.B
   )
  )
  n103 = MaterialFunctionCall(Float4=NamedRerouteUsage()) {materialFunction=&BreakOutFloat4Components}
  n104 = LinearInterpolate(
   A=LinearInterpolate(
    A=ScalarParameter() {defaultValue=1, group=Wrap_Liquid, parameterName=Liquid_Surface_Falloff_Weapon}
    B=ScalarParameter() {defaultValue=1, group=Wrap_Liquid, parameterName=Liquid_Surface_Falloff_Vehicle}
    Alpha=@n103.R
   )
   B=LinearInterpolate(
    A=ScalarParameter() {defaultValue=1, group=Wrap_Liquid, parameterName=Liquid_Surface_Falloff_Mech}
    B=ScalarParameter() {defaultValue=1, group=Wrap_Liquid, parameterName=Liquid_Surface_Falloff_Character}
    Alpha=@n103.A
   )
   Alpha=@n103.B
  )
  n105 = ScalarParameter() {defaultValue=1, group=Wrap_Liquid, parameterName=Liquid_Surface_Falloff_Pickaxe}
  n106 = NamedRerouteUsage()
  n107 = NamedRerouteUsage()
  n108 = Divide(
   A=@n107
   B=LinearInterpolate(
    A=LinearInterpolate(A=@n104, B=@n105, Alpha=@n106)
    B=ScalarParameter() {defaultValue=1, group=Wrap_Liquid, parameterName=Liquid_Surface_Falloff_Backpack}
    Alpha=NamedRerouteUsage()
   )
  )
  n109 = Multiply(A=Saturate(None=Power(Base=@n102)), B=ComponentMask(None=Saturate(None=@n108)) {b=true})
  n110 = ScalarParameter() {group=Wrap_Liquid, parameterName=Liquid_Surface__Z_Height_Offset Vehicle}
  n111 = ScalarParameter() {group=Wrap_Liquid, parameterName=Liquid_Surface__Z_Height_Offset VehicleSide}
  n112 = MaterialFunctionCall(Float4=@n39) {materialFunction=&BreakOutFloat4Components}
  n113 = LinearInterpolate(
   A=LinearInterpolate(
    A=ScalarParameter() {group=Wrap_Liquid, parameterName=Liquid_Surface__Z_Height_Offset Weapon}
    B=LinearInterpolate(A=@n110, B=@n111, Alpha=@n112.G)
    Alpha=@n112.R
   )
   B=LinearInterpolate(
    A=ScalarParameter() {group=Wrap_Liquid, parameterName=Liquid_Surface__Z_Height_Offset Mech}
    B=ScalarParameter() {group=Wrap_Liquid, parameterName=Liquid_Surface__Z_Height_Offset Character}
    Alpha=@n112.A
   )
   Alpha=@n112.B
  )
  n114 = ScalarParameter() {defaultValue=0.5, group=Wrap_Liquid, parameterName=Liquid_Surface_Master_liquid_height Vehicle, sortPriority=10}
  n115 = ScalarParameter() {defaultValue=0.5, group=Wrap_Liquid, parameterName=Liquid_Surface_Master_liquid_height VehicleSide, sortPriority=10}
  n116 = MaterialFunctionCall(Float4=@n39) {materialFunction=&BreakOutFloat4Components}
  n117 = LinearInterpolate(
   A=LinearInterpolate(
    A=ScalarParameter() {defaultValue=0.5, group=Wrap_Liquid, parameterName=Liquid_Surface_Master_liquid_height Weapon, sortPriority=10}
    B=LinearInterpolate(A=@n114, B=@n115, Alpha=@n116.G)
    Alpha=@n116.R
   )
   B=LinearInterpolate(
    A=ScalarParameter() {defaultValue=0.5, group=Wrap_Liquid, parameterName=Liquid_Surface_Master_liquid_height Mech, sortPriority=10}
    B=ScalarParameter() {defaultValue=0.5, group=Wrap_Liquid, parameterName=Liquid_Surface_Master_liquid_height Character, sortPriority=10}
    Alpha=@n116.A
   )
   Alpha=@n116.B
  )
  n118 = Saturate(None=Add(A=Add(A=@n113, B=@n7), B=@n117))
  n119 = Subtract(A=Time(), B=ScalarParameter() {defaultValue=-1000000000, group=Wrap_Liquid, parameterName=Liquid_TimeFrom-BP, sortPriority=100})
  n120 = ScalarParameter() {defaultValue=1, group=Wrap_Liquid, parameterName=Liquid_Time Druation, sortPriority=100}
  n121 = LinearInterpolate(
   A=ScalarParameter() {defaultValue=0.025, group=Wrap_Liquid, parameterName=Liquid_Surface_Distortion_StrenghtMX}
   B=ScalarParameter() {defaultValue=0.0001, group=Wrap_Liquid, parameterName=Liquid_Surface_Distortion_Strenght}
   Alpha=Saturate(None=Divide(A=@n119, B=@n120))
  )
  n122 = Multiply(A=@n51, B=@n121)
  n123 = Add(
   A=Add(A=@n118, B=ScalarParameter() {defaultValue=-0.01, group=Wrap_Liquid, parameterName=Liquid_Surface__Z_Height_Offset_SubSurface}) {constB=-0.01}
   B=@n122
  )
  n124 = CurveAtlasRowParameter(CurveTime=@n123) {
   atlas=/Game/Weapons/WeaponSkins/Wraps/Textures/CruveAtlas_Wraps_01
   curve=/Game/Weapons/WeaponSkins/Wraps/Textures/liquid/CC_liquid_mask
   defaultValue=38
   group=Wrap_Liquid
   parameterName=Liquid_ColorCurve_Surface
  }
  n125 = MaterialFunctionCall!(
   Index=ScalarParameter() {defaultValue=38, group=Wrap_Liquid, parameterName="Liquid_ColorCurve_Surface  Index"}
   CCinput
   CCinput_Alpha=@n124.A
   UV=@n123
  ) {materialFunction=&/MF_Global_CC_Tex}
  n126 = Multiply(
   A=OneMinus(None=@n125.Alpha)
   B=ScalarParameter() {defaultValue=0.25, group=Wrap_Liquid, parameterName=Liquid_Surface__SubSurface_Brightness}
  )
  n127 = Multiply(A=Saturate(None=@n109), B=Saturate(None=OneMinus(None=@n126)))
  n128 = CurveAtlasRowParameter(CurveTime=@n127) {
   atlas=&CruveAtlas_Wraps_01
   curve=&CC_liquid_mask
   defaultValue=38
   group=Wrap_Liquid
   parameterName=Liquid_ColorCurve_Surface
  }
  n129 = Add(A=@n118, B=@n122)
  n130 = CurveAtlasRowParameter(CurveTime=@n129) {
   atlas=&CruveAtlas_Wraps_01
   curve=&CC_liquid_mask
   defaultValue=38
   group=Wrap_Liquid
   parameterName=Liquid_ColorCurve_Surface
  }
  n131 = Constant3Vector() {constant=<rgba 0 0 1 1>}
  n132 = PreSkinnedLocalBounds()
  n133 = VectorParameter() {bUseCustomPrimitiveData=true, parameterName=Pliant_Crouch_Offset, primitiveDataIndex=7}
  n134 = Subtract(A=NamedRerouteUsage(), B=NamedRerouteUsage()) {desc=Only activate for player}
  n135 = LinearInterpolate(A=@n132.Min, B=Add(A=@n133.RGB, B=@n132.Min), Alpha=@n134)
  n136 = Subtract(A=LinearInterpolate(A=@n132.Max, B=Add(A=@n133.RGB, B=@n132.Max), Alpha=@n134), B=@n135)
  n137 = Multiply(A=AppendVector(A=Constant2Vector(), B=NamedRerouteUsage()), B=ComponentMask(None=@n136) {b=true})
  n138 = MaterialFunctionCall(Color=@n136) {materialFunction=&SplitComponents}
  n139 = MaterialFunctionCall() {materialFunction=&ObjectScale}
  n140 = Multiply(A=AppendVector(A=@n138.R, B=@n138.R), B=@n139.Scale Z)
  n141 = MaterialFunctionCall(
   Scale=@n140
   Object-SafeZone-Virtual-Pivot=NamedRerouteUsage()
   Object-SafeZone-Virtual-CH-Switch=NamedRerouteUsage()
   SinglePlane Normal Axis=@n131
   SinglePlane X Axis=Transform(None=Constant3Vector() {constant=<rgba 1 0 0 1>}) {transformSourceType=TRANSFORMSOURCE_Local}
   PlaneCenter_distortion=@n137
  ) {materialFunction=&MF_SafeZone_VirtualPlane}
  n142 = SphereMask(
   A=Add(A=NamedRerouteUsage(), B=@n141.Single_plane_UVs)
   B=Constant() {r=0.5}
   Radius=LinearInterpolate(
    A=ScalarParameter() {defaultValue=0.6762, group=Wrap_Liquid, parameterName=Liquid_WaterSurface_Radius}
    B=ScalarParameter() {defaultValue=0.6762, group=Wrap_Liquid, parameterName=Liquid_WaterSurface_Radius_Backpack}
    Alpha=NamedRerouteUsage()
   )
   Hardness=ScalarParameter() {defaultValue=0.75, group=Wrap_Liquid, parameterName=Liquid_WaterSurface_Hardness}
  )
  n143 = ScalarParameter() {defaultValue=0.75, group=Wrap_Liquid, parameterName=Liquid_WaterSurface_opacity}
  n144 = Add(
   A=Multiply(A=Multiply(A=@n141.Single_plane_Masking, B=@n141.Single_plane_Masking), B=Multiply(A=@n142, B=@n143))
   B=MaterialFunctionCall(
    Index=ScalarParameter() {defaultValue=38, group=Wrap_Liquid, parameterName="Liquid_ColorCurve_Surface  Index"}
    CCinput_Alpha=@n130.A
    UV=@n129
   ) {materialFunction=&MF_Global_CC_Tex}.Alpha
  )
  n145 = VectorParameter() {group=Wrap_Liquid, parameterName=Liquid_Glass_Edge_Color, sortPriority=0, defaultValue=<rgba 0.0677 0.0677 0.0677 0>}
  n146 = Saturate(None=Multiply(A=@n94) {constB=2})
  n147 = Add(
   A=LinearInterpolate(
    A=@n75.RGB
    B=MaterialFunctionCall(Index=@n76, CCinput=@n128, UV=@n127) {materialFunction=&MF_Global_CC_Tex}.3V
    Alpha=Saturate(None=@n144)
   )
   B=Multiply(A=@n145.RGB, B=@n146)
  )
  n148 = Subtract(A=FunctionInput(Preview=@n93.A) {bUsePreviewValueAsDefault=true, inputName=Gmap, inputType=FunctionInput_Scalar, sortPriority=24}) {constB=0.4}
  n149 = Add(A=@n16.UVs, B=@n21)
  n150 = ScalarParameter() {defaultValue=6, group=Wrap_Liquid, parameterName=Liquid_Bubble_B_TileRate_Weapon}
  n151 = MaterialFunctionCall(Float4=@n39) {materialFunction=&BreakOutFloat4Components}
  n152 = LinearInterpolate(
   A=ScalarParameter() {defaultValue=6, group=Wrap_Liquid, parameterName=Liquid_Bubble_B_TileRate_Vehicle}
   B=ScalarParameter() {defaultValue=6, group=Wrap_Liquid, parameterName=Liquid_Bubble_B_TileRate_VehicleSide}
   Alpha=@n151.G
  )
  n153 = ScalarParameter() {defaultValue=6, group=Wrap_Liquid, parameterName=Liquid_Bubble_B_TileRate_Mech}
  n154 = ScalarParameter() {defaultValue=6, group=Wrap_Liquid, parameterName=Liquid_Bubble_B_TileRate_Character}
  n155 = Multiply(
   A=LinearInterpolate(A=LinearInterpolate(A=@n150, B=@n152, Alpha=@n151.R), B=LinearInterpolate(A=@n153, B=@n154, Alpha=@n151.A), Alpha=@n151.B)
   B=@n149
  )
  n156 = ComponentMask(None=Multiply(A=Time(), B=VectorParameter() {group=Wrap_Liquid, parameterName=Liquid_Bubble_B_Speed, defaultValue=<rgba -0.066 0.03 -0.066 1>}.RGB)) {g=true, r=true}
  n157 = TextureObjectParameter() {group=Wrap_Liquid, parameterName=Liquid_Bubble__Tex, texture=&T_BlacksRGB}
  n158 = Multiply(A=TextureSample(UVs=Add(A=@n156, B=@n155), Tex=@n157).A) {constB=0.25}
  n159 = ScalarParameter() {defaultValue=6, group=Wrap_Liquid, parameterName=Liquid_Bubble_A_TileRate_Vehicle}
  n160 = ScalarParameter() {defaultValue=6, group=Wrap_Liquid, parameterName=Liquid_Bubble_A_TileRate_VehicleSide}
  n161 = MaterialFunctionCall(Float4=@n39) {materialFunction=&BreakOutFloat4Components}
  n162 = LinearInterpolate(
   A=LinearInterpolate(
    A=ScalarParameter() {defaultValue=6, group=Wrap_Liquid, parameterName=Liquid_Bubble_A_TileRate_Weapon}
    B=LinearInterpolate(A=@n159, B=@n160, Alpha=@n161.G)
    Alpha=@n161.R
   )
   B=LinearInterpolate(
    A=ScalarParameter() {defaultValue=6, group=Wrap_Liquid, parameterName=Liquid_Bubble_A_TileRate_Mech}
    B=ScalarParameter() {defaultValue=6, group=Wrap_Liquid, parameterName=Liquid_Bubble_A_TileRate_Character}
    Alpha=@n161.A
   )
   Alpha=@n161.B
  )
  n163 = Multiply(A=Time(), B=VectorParameter() {group=Wrap_Liquid, parameterName=Liquid_Bubble_A_Speed, defaultValue=<rgba -0.1 -0.03 -0.1 1>}.RGB)
  n164 = TextureSample(UVs=Add(A=ComponentMask(None=@n163) {g=true, r=true}, B=Multiply(A=@n162, B=@n149)), Tex=@n157)
  n165 = Multiply(A=Saturate(None=Multiply(A=@n148) {constB=1.5}), B=Saturate(None=Add(A=@n158, B=@n164.R)))
  n166 = Multiply(
   A=Multiply(A=@n109, B=Multiply(A=@n125.Alpha, B=@n165))
   B=VectorParameter() {group=Wrap_Liquid, parameterName=Liquid_Bubbles_Color, defaultValue=<rgba 0.1704 1.4403 1.5 1>}.RGB
  )
  n167 = Constant() {r=0.5}
  n168 = Constant() {r=1}
  n169 = Constant() {r=0.5}
  n170 = Constant() {r=0.25}
  n171 = SetMaterialAttributes!(
   MaterialAttributes=@n74.MaterialAttributes
   BaseColor=VectorParameter() {group=Wrap_Liquid, parameterName=Liquid_BaseColor, sortPriority=0, defaultValue=<rgba 0 0 0 1>}.RGB
   Roughness=ScalarParameter() {defaultValue=0.2, group=Wrap_Liquid, parameterName=Liquid_Roghness, sortPriority=0}
   Specular=ScalarParameter() {defaultValue=0.5, group=Wrap_Liquid, parameterName=Liquid_Specular, sortPriority=0}
   EmissiveColor=Multiply(
    A=Add(A=@n147, B=@n166)
    B=MaterialFunctionCall(Value_Morning=@n167, Value_Day=@n168, Value_Evening=@n169, Value_Night=@n170) {materialFunction=&Scalar_Time_of_Day}.Result
   )
   Metallic=ScalarParameter() {group=Wrap_Liquid, parameterName=Liquid_Metallic, sortPriority=0}
  ) {attributeSetTypes [
    #2
    #3
    #4
    #5
    #6
   ]}
  FunctionOutput(None=StaticSwitchParameter(True=@n171, False=@n74.MaterialAttributes) {group=Customization_Wrap, parameterName=Wrap_Liquid})
  n172 = VectorParameter() {group=Wrap_Liquid, parameterName=Liquid_Surface__Z_Center_Offset_Pliant_Pickaxe, defaultValue=<rgba -0.5 -0.5 -0.5 1>}
  LinearInterpolate(
   A=LinearInterpolate(
    A=VectorParameter() {group=Wrap_Liquid, parameterName=Liquid_Surface__Z_Center_Offset, defaultValue=<rgba -0.5 -0.5 -0.5 1>}.RGB
    B=@n172.RGB
    Alpha=NamedRerouteUsage()
   )
   B=@n71
   Alpha=NamedRerouteUsage()
  )
  LinearInterpolate(
   A=ScalarParameter() {defaultValue=0.2, group=Wrap_Liquid, parameterName=Liquid_Roghness_High, sortPriority=0}
   B=ScalarParameter() {defaultValue=0.2, group=Wrap_Liquid, parameterName=Liquid_Roghness_Low, sortPriority=0}
   Alpha=@n146
  )
  n173 = Multiply(A=@n139.Scale Z, B=AppendVector(A=@n138.G, B=@n138.G))
  LinearInterpolate(
   A=@n140
   B=@n173
   Alpha=MaterialFunctionCall(Float4=NamedRerouteUsage()) {materialFunction=&BreakOutFloat4Components}.A
  )
  MaterialFunctionCall(Color=@n39) {materialFunction=&SplitComponents}
  Multiply(A=AppendVector(A=Constant2Vector(), B=NamedRerouteUsage())) {constB=100}
  n174 = FunctionInput(Preview=Constant4Vector() {constant=<rgba 2.4195 25.3676 9.2814 -0.724>}.RGBA) {bUsePreviewValueAsDefault=true, inputName=Object-SafeZone-Virtual-Pivot, inputType=FunctionInput_Vector4, sortPriority=105}
  NamedRerouteDeclaration(None=@n174) {name=SafeZone-Virtual-Pivot, nodeColor=<rgba 0 0.8832 1 1>}
  NamedRerouteDeclaration(None=@n83.A) {name=IsCharcter, nodeColor=<rgba 1 0.4219 0 1>}
  n175 = FunctionInput() {bUsePreviewValueAsDefault=true, inputName=IsStatsB, inputType=FunctionInput_Vector2}
  NamedRerouteDeclaration(None=@n175) {name=IsState_B, nodeColor=<rgba 1 0.7031 0 1>}
  NamedRerouteDeclaration(None=FunctionInput() {bUsePreviewValueAsDefault=true, inputName=IsCharacter, inputType=FunctionInput_Scalar}) {name=IsCharacter, nodeColor=<rgba 0 0.9253 1 1>}
  NamedRerouteDeclaration(None=ComponentMask(None=@n175) {r=true}) {name=IsBackpack, nodeColor=<rgba 0 0.9253 1 1>}
  NamedRerouteDeclaration(None=ComponentMask(None=@n175) {g=true}) {name=IsPickaxe, nodeColor=<rgba 0 0.9253 1 1>}
  NamedRerouteDeclaration(None=@n90) {name=SafeZone-VP-CharcterSwitch, nodeColor=<rgba 0.1016 1 0 1>}
  NamedRerouteDeclaration(None=@n95) {name=Surface_Distortion, nodeColor=<rgba 0.5469 1 0 1>}
  n176 = Multiply(A=@n121, B=@n51) {constA=0.025}
  NamedRerouteDeclaration(None=@n176) {name=Surface_Distortion_B, nodeColor=<rgba 0.6172 0 1 1>}
  NamedRerouteDeclaration(None=@n39) {name=IsState, nodeColor=<rgba 1 0.7031 0 1>}
  n177 = PreSkinnedLocalBounds()
  n178 = Subtract(A=VertexInterpolator(VS=PreSkinnedPosition()).PS, B=@n177.Min)
  n179 = MaterialFunctionCall(Color=Subtract(A=@n177.Max, B=@n177.Max)) {materialFunction=&SplitComponents}
  n180 = Max(A=Max(A=@n179.R, B=@n179.R), B=@n179.B)
  NamedRerouteDeclaration(None=FunctionInput(Preview=Divide(A=@n178, B=@n180)) {bUsePreviewValueAsDefault=true, inputName=Calculated Bounds, sortPriority=10}) {name=Cal_Bounds, nodeColor=<rgba 0.9453 1 0 1>}
  Reroute(None=@n176)
  n181 = Constant3Vector() {constant=<rgba 0 0 1 0>}
  n182 = DotProduct(A=@n181, B=CameraPositionWS() {desc=Ray Origin})
  n183 = Multiply(A=@n3) {constB=0}
  n184 = AppendVector(A=Constant2Vector() {g=0.5, r=0.5}, B=@n117)
  n185 = Add(A=Add(A=Add(A=@n183, B=@n135), B=Multiply(A=@n184, B=@n136)), B=AppendVector(A=Constant2Vector(), B=Multiply(A=@n139.Scale Z) {constB=-1}))
  n186 = LinearInterpolate(A=@n185, B=Subtract(A=@n185, B=VectorParameter() {parameterName=Position_Adjust_Pliant_Pickaxe}.RGB), Alpha=NamedRerouteUsage())
  n187 = LinearInterpolate(
   A=Subtract(A=@n185, B=VectorParameter() {parameterName=Position_Adjust_Pliant_Character}.RGB)
   B=Subtract(A=@n185, B=VectorParameter() {parameterName=Position_Adjust_Pliant_Backpack}.RGB)
   Alpha=NamedRerouteUsage()
  )
  n188 = NamedRerouteUsage()
  n189 = Add(A=TransformPosition(None=LinearInterpolate(A=@n186, B=@n187, Alpha=@n188)) {transformType=TRANSFORMPOSSOURCE_World}, B=@n137)
  n190 = DotProduct(A=@n181, B=@n189) {desc=Distance to Plane}
  n191 = Subtract(A=WorldPosition().XYZ, B=CameraPositionWS()) {desc=Direction Vector}
  n192 = DotProduct(A=@n181, B=@n191) {desc=VectorDotPlaneNormal}
  n193 = Divide(A=Multiply(A=Subtract(A=@n182, B=@n190) {desc=Plane Distance}) {constB=-1}, B=@n192) {desc=Intersection Time1}
  n194 = Saturate(None=Divide(A=@n193) {constB=0})
  n195 = Saturate(None=OneMinus(None=@n192))
  n196 = Subtract(A=Add(A=CameraPositionWS() {desc=Ray Origin}, B=Multiply(A=@n193, B=@n191)) {desc=Intersection Position}, B=@n189)
  n197 = Transform(None=Constant3Vector() {constant=<rgba 0 1 0 0>}) {transformSourceType=TRANSFORMSOURCE_Local}
  n198 = Constant3Vector() {constant=<rgba 0 0.0002 -1 0>}
  n199 = Normalize(VectorInput=CrossProduct(
    A=@n197
    B=MaterialFunctionCall!(Vector1=@n181, Vector2=@n198) {materialFunction=/Engine/Functions/Engine_MaterialFunctions02/Math/CreateThirdOrthogonalVector}.Vector1
   )) {desc=Plane Y}
  n200 = Divide(A=AppendVector(A=DotProduct(A=@n197, B=@n196), B=DotProduct(A=@n199, B=@n196)), B=@n173)
  n201 = SphereMask(
   A=Add(A=NamedRerouteUsage(), B=Add(A=@n200) {constB=0.5})
   B=Constant() {r=0.5}
   Radius=LinearInterpolate(
    A=ScalarParameter() {defaultValue=0.6762, group=Wrap_Liquid, parameterName=Liquid_WaterSurface_Radius}
    B=ScalarParameter() {defaultValue=0.6762, group=Wrap_Liquid, parameterName=Liquid_WaterSurface_Radius_Backpack}
    Alpha=NamedRerouteUsage()
   )
   Hardness=ScalarParameter() {defaultValue=0.75, group=Wrap_Liquid, parameterName=Liquid_WaterSurface_Hardness}
  )
  n202 = ScalarParameter() {defaultValue=0.75, group=Wrap_Liquid, parameterName=Liquid_WaterSurface_opacity}
  Reroute(None=Multiply(A=Multiply(A=@n194, B=@n195), B=Multiply(A=@n201, B=@n202)))
  ScalarParameter() {defaultValue=3, group=Wrap_Liquid, parameterName=Liquid_Surface_Distortion_TileRate B}
  n203 = Multiply(A=@n3) {constB=0}
  n204 = VectorParameter() {group=Wrap_Liquid, parameterName=Liquid_Bubble_A_position, defaultValue=<rgba 0.4 0.4 0.4 0>}
  TransformPosition(None=Add(A=Add(A=@n203, B=@n12.Min), B=Multiply(A=@n204.RGB, B=@n13) {constA=0.5})) {transformType=TRANSFORMPOSSOURCE_World}
  n205 = Multiply(A=@n3) {constB=0}
  n206 = VectorParameter() {group=Wrap_Liquid, parameterName=Liquid_Bubble_B_position, defaultValue=<rgba 0.6 0.6 0.6 1>}
  TransformPosition(None=Add(A=Add(A=@n205, B=@n25.Min), B=Multiply(A=@n206.RGB, B=@n26) {constA=0.5})) {transformType=TRANSFORMPOSSOURCE_World}
  Transform(None=@n131) {transformSourceType=TRANSFORMSOURCE_Local}
  Transform(None=@n107) {transformSourceType=TRANSFORMSOURCE_Local}
 ]}
