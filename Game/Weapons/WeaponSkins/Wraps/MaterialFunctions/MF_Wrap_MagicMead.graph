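# Material function for the Magic Meadow weapon wrap, switched over the incoming material by a static switch, adding animated flames, embers and icons.
# IsState inputs select per-object projection settings for weapon, vehicle, mech, character, backpack and pickaxe.
# Scrolling flame textures height-blend against curve-atlas gradients, and masks combine base and accent layers.
# Blueprint-set timestamps drive fading fire, muzzle and kill glows, with time-of-day scaled emissive.
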MaterialFunction /Game/Weapons/WeaponSkins/Wraps/MaterialFunctions/MF_Wrap_MagicMead
graph {nodes [
  n1 = Add(
   A=Multiply(A=Time(), B=VectorParameter() {group=Wrap_MagicMeadow_Fire_Uvs, parameterName=MMead_Fire_A_Speed, defaultValue=<rgba 0.5 0.053 0 1>}.RGB)
   B=NamedRerouteUsage()
  )
  n2 = Time()
  n3 = Add(
   A=ScalarParameter() {defaultValue=-9999999, group=Wrap_MagicMeadow_Reactive, parameterName=MMead_Time_BP}
   B=ScalarParameter() {defaultValue=0.25, group=Wrap_MagicMeadow_Reactive, parameterName=MMead_Fire_Delay_Offset, sortPriority=200}
  )
  n4 = Saturate(None=Divide(
    A=Subtract(A=@n2, B=@n3)
    B=ScalarParameter() {defaultValue=0.25, group=Wrap_MagicMeadow_Reactive, parameterName=MMead_Fire_Duration_Offset, sortPriority=200}
   ))
  n5 = ScalarParameter() {defaultValue=1, group=Wrap_MagicMeadow_Reactive, parameterName=MMead_Fire_Duration_offset curve, sortPriority=200}
  n6 = AppendVector(
   A=LinearInterpolate(
    A=ScalarParameter() {group=Wrap_MagicMeadow_Reactive, parameterName=MMead_Fire_Start_Offset, sortPriority=200}
    B=ScalarParameter() {group=Wrap_MagicMeadow_Reactive, parameterName=MMead_Fire_End_Offset, sortPriority=200}
    Alpha=Power(Base=@n4, Exp=@n5)
   )
   B=Constant2Vector()
  )
  Add(A=@n1, B=@n6)
  n7 = Time()
  n8 = Sine(None=Multiply(A=@n7, B=ScalarParameter() {defaultValue=1.5, parameterName=MMead_Icon_A_Flicker_Speed}))
  n9 = Sine(None=Multiply(A=@n7, B=ScalarParameter() {defaultValue=4, parameterName=MMead_Icon_B_Flicker_Speed}))
  Add(A=@n8, B=@n9)
  n10 = Time()
  Add(A=@n10, B=@n6)
  ComponentMask(None=Constant3Vector() {constant=<rgba 1 1 0 0>}) {g=true, r=true}
  n11 = NamedRerouteUsage()
  n12 = PreSkinnedLocalBounds()
  n13 = PreSkinnedLocalBounds()
  n14 = Subtract(A=Add(A=@n13.Max, B=NamedRerouteUsage()), B=Add(A=@n13.Min, B=NamedRerouteUsage()))
  n15 = Add(A=@n12.Min, B=@n14)
  n16 = Subtract(A=Subtract(A=@n12.Max, B=@n12.Max), B=@n14)
  n17 = NamedRerouteDeclaration(None=Divide(A=Subtract(A=@n11, B=@n15), B=@n16)) {name=Bounds, nodeColor=<rgba 1 0 0.7031 1>}
  n18 = ComponentMask(None=@n17) {g=true}
  n19 = ComponentMask(None=Add(A=NamedRerouteUsage()) {constB=-1}) {g=true}
  Divide(A=@n18, B=@n19) {constB=0.001}
  n20 = TextureSample() {samplerType=SAMPLERTYPE_Normal, texture=/Game/Weapons/FORT_Rifles/T_SCAR_N}
  n21 = Constant3Vector() {constant=<rgba 1 0 0 0>}
  n22 = GetMaterialAttributes(None=FunctionInput(Preview=SetMaterialAttributes!(MaterialAttributes, Normal=@n20.RGB, BaseColor=@n21) {attributeSetTypes [#1, #2]}) {inputName=Material, inputType=FunctionInput_MaterialAttributes}) {attributeGetTypes [
    #2
    #3
    #1
    #4
   ]}
  FunctionOutput(None=StaticSwitchParameter(True=NamedRerouteUsage(), False=NamedRerouteDeclaration(None=@n22.MaterialAttributes) {name=Root_Material, nodeColor=<rgba 0.7812 1 0 1>}) {group=Customization_Wrap, parameterName=Wrap_MagicMeadow})
  n23 = VectorParameter() {group=Wrap_MagicMeadow_Fire_Uvs, parameterName=MMead_Rotation_Mech, defaultValue=<rgba 0.25 0.25 0 1>}
  n24 = VectorParameter() {group=Wrap_MagicMeadow_Fire_Uvs, parameterName=MMead_Rotation_Character, defaultValue=<rgba 0.25 0.25 0 1>}
  n25 = MaterialFunctionCall(Float4=NamedRerouteUsage()) {materialFunction=&BreakOutFloat4Components}
  LinearInterpolate(A=@n23.RGB, B=@n24.RGB, Alpha=@n25.A)
  n26 = ScalarParameter() {group=Wrap_MagicMeadow_Icon_B_Uvs, parameterName=MMead_IconB_Rotation_Mech}
  n27 = MaterialFunctionCall(Float4=NamedRerouteUsage()) {materialFunction=&BreakOutFloat4Components}
  LinearInterpolate(A=@n26, B=ScalarParameter() {defaultValue=1, group=Wrap_MagicMeadow_Icon_B_Uvs, parameterName=MMead_IconB_Rotation_Character}, Alpha=@n27.A)
  n28 = ScalarParameter() {group=Wrap_MagicMeadow_Icon_B_Uvs, parameterName=MMead_IconB_Rotation_Vehicle}
  n29 = ScalarParameter() {group=Wrap_MagicMeadow_Icon_B_Uvs, parameterName=MMead_IconB_Rotation_Vehicle_Side}
  n30 = LinearInterpolate(
   A=LinearInterpolate(
    A=ScalarParameter() {group=Wrap_MagicMeadow_Icon_B_Uvs, parameterName=MMead_IconB_Rotation_Weapon}
    B=LinearInterpolate(A=@n28, B=@n29, Alpha=@n27.G)
    Alpha=@n27.R
   )
   B=@n26
   Alpha=@n27.B
  )
  n31 = MaterialFunctionCall(Float2=NamedRerouteUsage()) {materialFunction=&BreakOutFloat2Components}
  LinearInterpolate(
   A=LinearInterpolate(A=@n30, B=ScalarParameter() {defaultValue=1, group=Wrap_MagicMeadow_Icon_B_Uvs, parameterName=MMead_IconB_Rotation_Backpack}, Alpha=@n31.R)
   B=ScalarParameter() {defaultValue=1, group=Wrap_MagicMeadow_Icon_B_Uvs, parameterName=MMead_IconB_Rotation_Pickaxe}
   Alpha=@n31.G
  )
  n32 = VectorParameter() {group=Wrap_MagicMeadow_Fire_Uvs, parameterName=MMead_Rotation_Vehicle, defaultValue=<rgba 0 0 0 1>}
  n33 = VectorParameter() {group=Wrap_MagicMeadow_Fire_Uvs, parameterName=MMead_Rotation_Vehicle_Side, defaultValue=<rgba 0 0 0 1>}
  n34 = LinearInterpolate(
   A=LinearInterpolate(
    A=VectorParameter() {group=Wrap_MagicMeadow_Fire_Uvs, parameterName=MMead_Rotation_Weapon, defaultValue=<rgba 0 0 -0.25 1>}.RGB
    B=LinearInterpolate(A=@n32.RGB, B=@n33.RGB, Alpha=@n25.G)
    Alpha=@n25.R
   )
   B=@n23.RGB
   Alpha=@n25.B
  )
  n35 = MaterialFunctionCall(Float2=NamedRerouteUsage()) {materialFunction=&BreakOutFloat2Components}
  LinearInterpolate(
   A=LinearInterpolate(
    A=@n34
    B=VectorParameter() {group=Wrap_MagicMeadow_Fire_Uvs, parameterName=MMead_Rotation_Backpack, defaultValue=<rgba 0 0 0 1>}.RGB
    Alpha=@n35.R
   )
   B=VectorParameter() {group=Wrap_MagicMeadow_Fire_Uvs, parameterName=MMead_Rotation_Pickaxe, defaultValue=<rgba 0 0 0 1>}.RGB
   Alpha=@n35.G
  )
  n36 = ScalarParameter() {group=Wrap_MagicMeadow_Fire, parameterName=MMead_Grad_Rotation_Mech}
  n37 = MaterialFunctionCall(Float4=NamedRerouteUsage()) {materialFunction=&BreakOutFloat4Components}
  LinearInterpolate(A=@n36, B=ScalarParameter() {defaultValue=1, group=Wrap_MagicMeadow_Fire, parameterName=MMead_Grad_Rotation_Character}, Alpha=@n37.A)
  n38 = ScalarParameter() {group=Wrap_MagicMeadow_Fire, parameterName=MMead_Grad_Rotation_Vehicle}
  n39 = ScalarParameter() {group=Wrap_MagicMeadow_Fire, parameterName=MMead_Grad_Rotation_Vehicle_Side}
  n40 = LinearInterpolate(
   A=LinearInterpolate(
    A=ScalarParameter() {group=Wrap_MagicMeadow_Fire, parameterName=MMead_Grad_Rotation_Weapon}
    B=LinearInterpolate(A=@n38, B=@n39, Alpha=@n37.G)
    Alpha=@n37.R
   )
   B=@n36
   Alpha=@n37.B
  )
  n41 = MaterialFunctionCall(Float2=NamedRerouteUsage()) {materialFunction=&BreakOutFloat2Components}
  LinearInterpolate(
   A=LinearInterpolate(A=@n40, B=ScalarParameter() {defaultValue=1, group=Wrap_MagicMeadow_Fire, parameterName=MMead_Grad_Rotation_Backpack}, Alpha=@n41.R)
   B=ScalarParameter() {defaultValue=1, group=Wrap_MagicMeadow_Fire, parameterName=MMead_Grad_Rotation_Pickaxe}
   Alpha=@n41.G
  )
  n42 = ScalarParameter() {group=Wrap_MagicMeadow_Icon_A_Uvs, parameterName=MMead_IconA_Rotation_Mech}
  n43 = MaterialFunctionCall(Float4=NamedRerouteUsage()) {materialFunction=&BreakOutFloat4Components}
  LinearInterpolate(A=@n42, B=ScalarParameter() {group=Wrap_MagicMeadow_Icon_A_Uvs, parameterName=MMead_IconA_Rotation_Character}, Alpha=@n43.A)
  n44 = ScalarParameter() {group=Wrap_MagicMeadow, parameterName=MMead_Metlaic}
  LinearInterpolate(A=@n44, Alpha=NamedRerouteUsage()) {constB=0}
  n45 = ScalarParameter() {group=Wrap_MagicMeadow_Icon_A_Uvs, parameterName=MMead_IconA_Rotation_Vehicle}
  n46 = ScalarParameter() {group=Wrap_MagicMeadow_Icon_A_Uvs, parameterName=MMead_IconA_Rotation_Vehicle_Side}
  n47 = LinearInterpolate(
   A=LinearInterpolate(
    A=ScalarParameter() {group=Wrap_MagicMeadow_Icon_A_Uvs, parameterName=MMead_IconA_Rotation_Weapon}
    B=LinearInterpolate(A=@n45, B=@n46, Alpha=@n43.G)
    Alpha=@n43.R
   )
   B=@n42
   Alpha=@n43.B
  )
  n48 = MaterialFunctionCall(Float2=NamedRerouteUsage()) {materialFunction=&BreakOutFloat2Components}
  LinearInterpolate(
   A=LinearInterpolate(A=@n47, B=ScalarParameter() {group=Wrap_MagicMeadow_Icon_A_Uvs, parameterName=MMead_IconA_Rotation_Backpack}, Alpha=@n48.R)
   B=ScalarParameter() {group=Wrap_MagicMeadow_Icon_A_Uvs, parameterName=MMead_IconA_Rotation_Pickaxe}
   Alpha=@n48.G
  )
  n49 = MaterialFunctionCall(Color=@n16) {materialFunction=&SplitComponents}
  Max(A=Max(A=@n49.R, B=@n49.R), B=@n49.B)
  Multiply(A=MaterialFunctionCall(Value_Morning=Constant() {r=0.5}, Value_Day=Constant() {r=1}, Value_Evening=Constant() {r=0.5}, Value_Night=Constant() {r=0.25}) {materialFunction=&Scalar_Time_of_Day}.Result)
  n50 = NamedRerouteUsage()
  n51 = MaterialFunctionCall!(
   Normal_Map
   Offset Rt
   Pivot
   Caculated_Bounds=NamedRerouteUsage()
   Pre_Skinned_Normals=NamedRerouteUsage()
   is_State=NamedRerouteUsage()
   UV_Rotation=NamedRerouteUsage()
   Uvs_Offset=ComponentMask(None=@n50) {b=true, g=true, r=true}
   Uvs_TileRate=ComponentMask(None=@n50) {a=true}
   Uvs_Offset_Pivot
   For_Safe_Zones=StaticBool() {value=true}
   Rotation_Comp
   Clamp Y
   Clamp X
   Clamp Padding_Comp
   No_Rotation
  ) {materialFunction=&/MF_Projection_Side_Normal_V2}
  n52 = Saturate(None=@n51.UVs)
  n53 = TextureSample(UVs=@n52, Tex=NamedRerouteUsage())
  Multiply(A=@n53.R) {constB=0.5}
  n54 = ComponentMask(None=VectorParameter() {group=Wrap_MagicMeadow_Icon_A, parameterName=MMead_IconA_Msk_Offset, defaultValue=<rgba -0.4 -0.5 0 1>}.RGB) {g=true, r=true}
  n55 = SmoothStep(
   Min=ScalarParameter() {defaultValue=0.6, group=Wrap_MagicMeadow_Icon_A, parameterName=MMead_IconA_Msking Size}
   Value=MaterialFunctionCall(Vector 2=Add(A=@n51.UVs, B=@n54) {constB=-0.5}) {materialFunction=/Engine/Functions/Engine_MaterialFunctions02/Utility/VectorLength}.V2 Length
  ) {constMax=0, constMin=1.5}
  n56 = Multiply(A=@n55) {constB=10}
  n57 = Add(A=@n52, B=ScalarParameter() {defaultValue=-0.1, group=Wrap_MagicMeadow_Icon_A, parameterName=MMead_IconA_Half_Msk}) {constB=-0.5}
  n58 = Saturate(None=Divide(
    A=ComponentMask(None=@n57) {r=true}
    B=ScalarParameter() {defaultValue=0.5, group=Wrap_MagicMeadow_Icon_A, parameterName=MMead_IconA_Half_Msk_Falloff}
   ) {constB=1})
  n59 = OneMinus(None=Saturate(None=LinearInterpolate(A=@n55, B=@n56, Alpha=@n58)))
  n60 = NamedRerouteUsage()
  n61 = MaterialFunctionCall(
   Caculated_Bounds=NamedRerouteUsage()
   Pre_Skinned_Normals=NamedRerouteUsage()
   is_State=NamedRerouteUsage()
   UV_Rotation=NamedRerouteUsage()
   Uvs_Offset=ComponentMask(None=@n60) {b=true, g=true, r=true}
   Uvs_TileRate=ComponentMask(None=@n60) {a=true}
   For_Safe_Zones=StaticBool() {value=true}
  ) {materialFunction=&MF_Projection_Side_Normal_V2}
  n62 = TextureSample(UVs=Saturate(None=@n61.UVs), Tex=NamedRerouteUsage()) {samplerType=SAMPLERTYPE_Masks}
  n63 = SmoothStep(Value=@n62.B) {constMax=0.5, constMin=0.2}
  n64 = OneMinus(None=Multiply(A=@n63, B=ScalarParameter() {defaultValue=0.9, group=Wrap_MagicMeadow, parameterName=MMeads_Icon_B_Msk_Strength}) {constB=0.5})
  Multiply(A=@n59, B=@n64)
  n65 = OneMinus(None=NamedRerouteUsage())
  Multiply(A=@n65)
  n66 = LinearInterpolate(
   A=LinearInterpolate(
    A=VectorParameter() {group=Wrap_MagicMeadow, parameterName=MMead_BckGB_Color, defaultValue=<rgba 0.035 0.035 0.035 1>}.RGB
    B=VectorParameter() {group=Wrap_MagicMeadow, parameterName=MMead_BckG_Color, defaultValue=<rgba 0.035 0.035 0.035 1>}.RGB
    Alpha=NamedRerouteUsage()
   )
   B=VectorParameter() {group=Wrap_MagicMeadow, parameterName=MMead_Icon_B_Color_A, defaultValue=<rgba 0 0 0 1>}.RGB
   Alpha=@n62.G
  )
  n67 = VectorParameter() {group=Wrap_MagicMeadow, parameterName=MMead_Icon_B_Color_B, defaultValue=<rgba 0.2292 0.0075 0 1>}
  n68 = Multiply(A=@n61.UVs) {constB=0.001}
  n69 = Add(
   A=Add(A=@n62.A) {constB=-0.5}
   B=Add(A=ComponentMask(None=@n68) {r=true}, B=ScalarParameter() {group=Wrap_MagicMeadow_Reactive, parameterName=MMead_KillCount_Offset})
  )
  n70 = CurveAtlasRowParameter(CurveTime=@n69) {
   atlas=/Game/Weapons/WeaponSkins/Wraps/Textures/CruveAtlas_Wraps_01
   curve=/Game/Weapons/WeaponSkins/Wraps/Textures/0_ColorCurves/CC_Flame_Msk
   defaultValue=110
   group=Wrap_MagicMeadow_Fire
   parameterName=MMead_CC_Fire_Msk
  }
  n71 = Multiply(
   A=@n62.R
   B=MaterialFunctionCall!(
    Index=ScalarParameter() {defaultValue=110, group=Wrap_MagicMeadow_Fire, parameterName="MMead_CC_Fire_Msk  Index"}
    CCinput
    CCinput_Alpha=@n70.A
    UV=@n69
   ) {materialFunction=&/MF_Global_CC_Tex}.Alpha
  )
  Multiply(A=NamedRerouteDeclaration(None=LinearInterpolate(A=@n66, B=@n67.RGB, Alpha=@n71)) {name=Icon_Base_Color, nodeColor=<rgba 1 0.9141 0 1>})
  Multiply(A=Power(Base=NamedRerouteUsage()))
  n72 = LinearInterpolate(Alpha=Add(A=ConstantBiasScale(None=@n8) {bias=1.5, scale=0.25}, B=ConstantBiasScale(None=@n9) {bias=1.5, scale=0.25})) {constA=0.5, constB=1.5}
  Multiply(A=MaterialFunctionCall!(X range Y range, Gradient Height, ShowGrid?, Thickness, Color, "f(x)"=@n72) {materialFunction=/Engine/Functions/Engine_MaterialFunctions01/Debug/PlotFunctionOnGraph}."f(x) = y")
  n73 = VectorParameter() {group=Wrap_MagicMeadow_Accent, parameterName=MMead_Accent_BaseColor, defaultValue=<rgba 0.8125 0.8125 0.8125 1>}
  n74 = SetMaterialAttributes!(
   MaterialAttributes=NamedRerouteUsage()
   BaseColor=@n73.RGB
   Roughness=ScalarParameter() {defaultValue=0.35, group=Wrap_MagicMeadow_Accent, parameterName=MMead_Accent_Roughness}
   Specular=ScalarParameter() {defaultValue=0.5, group=Wrap_MagicMeadow_Accent, parameterName=MMead_Accent_Spec}
   Metallic=ScalarParameter() {defaultValue=1, group=Wrap_MagicMeadow_Accent, parameterName=MMead_Accent_Metalic}
   Normal
   EmissiveColor
  ) {attributeSetTypes [
    #2
    #3
    #5
    #4
    #1
    #6
   ]}
  NamedRerouteDeclaration(None=@n74) {name=Mat_Accent, nodeColor=<rgba 0 0.3906 1 1>}
  n75 = TextureSample() {samplerType=SAMPLERTYPE_Masks, texture=/Game/Weapons/FORT_Rifles/Textures/T_Scar_CM}
  n76 = FunctionInput(Preview=@n75.G) {bUsePreviewValueAsDefault=true, inputName=Mask Crevise, inputType=FunctionInput_Scalar, sortPriority=24, previewValue=<1 0 0 1>}
  NamedRerouteDeclaration(None=@n76) {name=Msk_Curvature, nodeColor=<rgba 0 1 0.7422 1>}
  n77 = MaterialFunctionCall(
   Caculated_Bounds=NamedRerouteUsage()
   Pre_Skinned_Normals=NamedRerouteUsage()
   is_State=NamedRerouteUsage()
   UV_Rotation=NamedRerouteUsage()
   Uvs_Offset=Constant3Vector() {constant=<rgba -1 0 0 0>}
  ) {materialFunction=&MF_Projection_Side_Normal_V2}
  n78 = NamedRerouteUsage()
  n79 = ComponentMask(None=Abs(None=Multiply(A=@n77.UVs, B=@n78))) {r=true}
  n80 = CurveAtlasRowParameter(CurveTime=@n79) {
   atlas=&CruveAtlas_Wraps_01
   curve=/Game/Weapons/WeaponSkins/Wraps/Textures/0_ColorCurves/CC_Flame_Colors
   defaultValue=109
   group=Wrap_MagicMeadow_Fire
   parameterName=MMead_CC_Fire_Color
  }
  n81 = MaterialFunctionCall(
   Index=ScalarParameter() {defaultValue=109, group=Wrap_MagicMeadow_Fire, parameterName="MMead_CC_Fire_Color  Index"}
   CCinput=@n80
   CCinput_Alpha=@n80
   UV=@n79
  ) {materialFunction=&MF_Global_CC_Tex}
  NamedRerouteDeclaration(None=@n81.Alpha) {name=CC_Fire_Color_Msk, nodeColor=<rgba 0.4766 1 0 1>}
  NamedRerouteDeclaration(None=@n34) {name=Uvs_TSide_Rotation, nodeColor=<rgba 0 0.9297 1 1>}
  NamedRerouteDeclaration(None=@n47) {name=UVs_IconA_Rotation, nodeColor=<rgba 0 0.625 1 1>}
  n82 = Constant4Vector() {constant=<rgba 0 0 0 0.5>}
  NamedRerouteDeclaration(None=FunctionInput(Preview=@n82.RGBA) {bUsePreviewValueAsDefault=true, inputName=Object-SafeZone-Back, inputType=FunctionInput_Vector4, sortPriority=200}) {name=SafeZone_Back, nodeColor=<rgba 0 0.9297 1 1>}
  NamedRerouteDeclaration(None=@n53.RGBA) {name=IconA_RGBA, nodeColor=<rgba 0.9219 1 0 1>}
  NamedRerouteDeclaration(None=Multiply(A=@n55) {constB=0.1}) {name=IconA_Wrarp_Amount, nodeColor=<rgba 0.9453 0 1 1>}
  n83 = NamedRerouteUsage()
  n84 = ScalarParameter() {defaultValue=0.25, group=Wrap_MagicMeadow_Reactive, parameterName=MMead_Embers_Duration}
  n85 = LinearInterpolate(
   A=ScalarParameter() {defaultValue=1, group=Wrap_MagicMeadow_Reactive, parameterName=MMead_Embers_Start_Brightness}
   B=ScalarParameter() {group=Wrap_MagicMeadow_Reactive, parameterName=MMead_Embers_End_Brightness}
   Alpha=Saturate(None=Divide(A=@n83, B=@n84))
  )
  n86 = NamedRerouteUsage()
  n87 = ScalarParameter() {defaultValue=20, group=Wrap_MagicMeadow_Fire, parameterName=MMead_Flame_Contrast}
  n88 = CurveAtlasRowParameter(CurveTime=@n79) {
   atlas=&CruveAtlas_Wraps_01
   curve=&CC_Flame_Msk
   defaultValue=110
   group=Wrap_MagicMeadow_Fire
   parameterName=MMead_CC_Fire_Msk
  }
  MaterialExpressionMaterialFunctionCall_38 = MaterialFunctionCall(Index=ScalarParameter() {defaultValue=110, group=Wrap_MagicMeadow_Fire, parameterName="MMead_CC_Fire_Msk  Index"}, CCinput=@n88, UV=@n79) {materialFunction=&MF_Global_CC_Tex}
  n89 = Convert(None=@MaterialExpressionMaterialFunctionCall_38.3V) {
   nodeName=Break Float3
   convertInputs [{
     type=Vector3
     expressionInput {
      expression=&self:MaterialExpressionMaterialFunctionCall_38
      inputName=None
      mask=0
      maskA=0
      maskB=0
      maskG=0
      maskR=0
      outputIndex=0
     }
    }]
   convertOutputs [{type=Scalar}, {type=Scalar}, {type=Scalar}]
  }
  n90 = Multiply(A=NamedRerouteUsage(), B=NamedRerouteUsage())
  n91 = Multiply(A=@n89, B=@n90)
  n92 = MaterialFunctionCall(Transition Phase=@n91, Height Texture=ComponentMask(None=@n86) {r=true}, Contrast=@n87) {materialFunction=&HeightLerp}
  n93 = NamedRerouteUsage()
  n94 = MaterialFunctionCall(Transition Phase=@n91, Height Texture=ComponentMask(None=@n93) {r=true}, Contrast=@n87) {materialFunction=&HeightLerp}
  n95 = Saturate(None=Add(A=@n92.Alpha, B=@n94.Alpha))
  n96 = MaterialFunctionCall(Transition Phase=Multiply(A=@n89, B=@n90), Height Texture=ComponentMask(None=@n93) {a=true}, Contrast=@n87) {materialFunction=&HeightLerp}
  n97 = MaterialFunctionCall(Transition Phase=Multiply(A=@n89, B=@n90), Height Texture=ComponentMask(None=@n86) {a=true}, Contrast=@n87) {materialFunction=&HeightLerp}
  n98 = Add(A=@n96.Alpha, B=@n97.Alpha)
  n99 = Multiply(A=Add(A=Multiply(A=@n85, B=@n95), B=Saturate(None=@n98)), B=@n65)
  n100 = MaterialFunctionCall(Float4=NamedRerouteUsage()) {materialFunction=&BreakOutFloat4Components}
  n101 = Multiply(
   A=OneMinus(None=Saturate(None=@n99))
   B=OneMinus(None=NamedRerouteDeclaration(None=@n100.A) {name=icon_A_Msk, nodeColor=<rgba 0.8047 0 1 1>})
  )
  n102 = Multiply(A=@n85, B=NamedRerouteUsage())
  n103 = VectorParameter() {group=Wrap_MagicMeadow_Fire, parameterName=MMead_Embers_Color_B, defaultValue=<rgba 0.651 0.1312 0 1>}
  n104 = Multiply(
   A=NamedRerouteUsage()
   B=LinearInterpolate(
    A=Multiply(A=@n103.RGB, B=@n92.Alpha)
    B=VectorParameter() {group=Wrap_MagicMeadow_Fire, parameterName=MMead_Embers_Color_A, defaultValue=<rgba 0.651 0.4602 0 1>}.RGB
    Alpha=@n94.Alpha
   )
  )
  n105 = NamedRerouteUsage()
  n106 = Multiply(A=Multiply(A=@n105, B=@n97.Alpha))
  n107 = Multiply(A=@n96.Alpha) {constB=0.25}
  n108 = Multiply(A=@n65, B=Add(A=Multiply(A=@n102, B=@n104), B=LinearInterpolate(A=@n106, B=@n105, Alpha=@n107)))
  n109 = LinearInterpolate(
   A=VectorParameter() {group=Wrap_MagicMeadow_Icon_A, parameterName=MMead_Icon_A_Color, defaultValue=<rgba 0.6094 0.0858 0 1>}.RGB
   B=VectorParameter() {group=Wrap_MagicMeadow_Icon_A, parameterName=MMead_Icon_B_Color, defaultValue=<rgba 1 0.8982 0 1>}.RGB
   Alpha=@n100.R
  )
  n110 = ScalarParameter() {defaultValue=5, group=Wrap_MagicMeadow_Reactive, parameterName=MMead_Icon_A_Start_Brighness}
  n111 = ScalarParameter() {defaultValue=1, group=Wrap_MagicMeadow_Reactive, parameterName=MMead_Icon_A_End_Brighness}
  n112 = NamedRerouteUsage()
  n113 = Multiply(A=@n100.A, B=Multiply(A=Multiply(A=@n109, B=@n72), B=LinearInterpolate(A=@n110, B=@n111, Alpha=@n112)))
  n114 = Multiply(A=NamedRerouteUsage(), B=NamedRerouteUsage())
  n115 = VectorParameter() {group=Wrap_MagicMeadow_Reactive, parameterName=MMead_Kill_Reactive_Color, defaultValue=<rgba 0.7708 0 0.0136 1>}
  n116 = SetMaterialAttributes(
   MaterialAttributes=NamedRerouteUsage()
   BaseColor=Multiply(A=NamedRerouteUsage(), B=Saturate(None=@n101))
   Roughness=NamedRerouteUsage()
   Specular=ScalarParameter() {defaultValue=0.5, group=Wrap_MagicMeadow, parameterName=MMead_Specular}
   Metallic=@n44
   EmissiveColor=Add(A=Add(A=@n108, B=@n113), B=Multiply(A=@n114, B=@n115.RGB))
  ) {attributeSetTypes [
    #2
    #3
    #5
    #4
    #1
    #6
   ]}
  NamedRerouteDeclaration(None=@n116) {name=Mat_Base, nodeColor=<rgba 0 0.3906 1 1>}
  n117 = VectorParameter() {group=Wrap_MagicMeadow_Icon_A_Uvs, parameterName=MMead_IconA_Offset_Scale_Weapon, defaultValue=<rgba -0.43 -0.01 0 0.1>}
  n118 = VectorParameter() {group=Wrap_MagicMeadow_Icon_A_Uvs, parameterName=MMead_IconA_Offset_Scale_Vehicle, defaultValue=<rgba 0 0 0 1>}
  n119 = AppendVector(A=@n118.RGB, B=@n118.RGB)
  n120 = VectorParameter() {group=Wrap_MagicMeadow_Icon_A_Uvs, parameterName=MMead_IconA_Offset_Scale_VehicleSide, defaultValue=<rgba 0 0 0 1>}
  n121 = AppendVector(A=@n120.RGB, B=@n120.RGB)
  n122 = MaterialFunctionCall(Float4=NamedRerouteUsage()) {materialFunction=&BreakOutFloat4Components}
  n123 = VectorParameter() {group=Wrap_MagicMeadow_Icon_A_Uvs, parameterName=MMead_IconA_Offset_Scale_Mech, defaultValue=<rgba 0 0 0 1>}
  n124 = VectorParameter() {group=Wrap_MagicMeadow_Icon_A_Uvs, parameterName=MMead_IconA_Offset_Scale_Character, defaultValue=<rgba 0 0 0 1>}
  n125 = LinearInterpolate(
   A=LinearInterpolate(A=AppendVector(A=@n117.RGB, B=@n117.RGB), B=LinearInterpolate(A=@n119, B=@n121, Alpha=@n122.G), Alpha=@n122.R)
   B=LinearInterpolate(A=AppendVector(A=@n123.RGB, B=@n123.RGB), B=AppendVector(A=@n124.RGB, B=@n124.RGB), Alpha=@n122.A)
   Alpha=@n122.B
  )
  n126 = VectorParameter() {group=Wrap_MagicMeadow_Icon_A_Uvs, parameterName=MMead_IconA_Offset_Scale_Backpack, defaultValue=<rgba 0 0 0 1>}
  n127 = AppendVector(A=@n126.RGB, B=@n126.RGB)
  n128 = MaterialFunctionCall(Float2=NamedRerouteUsage()) {materialFunction=&BreakOutFloat2Components}
  n129 = VectorParameter() {group=Wrap_MagicMeadow_Icon_A_Uvs, parameterName=MMead_IconA_Offset_Scale_Pickaxe, defaultValue=<rgba 0 0 0 1>}
  NamedRerouteDeclaration(None=LinearInterpolate(A=LinearInterpolate(A=@n125, B=@n127, Alpha=@n128.R), B=AppendVector(A=@n129.RGB, B=@n129.RGB), Alpha=@n128.G)) {name=UVs_IconA_Offset_Scale, nodeColor=<rgba 0 0.625 1 1>}
  NamedRerouteDeclaration(None=TextureObjectParameter() {group=Wrap_MagicMeadow_Fire, parameterName=MMead__tex_Msk2, texture=&T_BlacksRGB}) {name=Tex_Msk2, nodeColor=<rgba 1 0 0.4922 1>}
  NamedRerouteDeclaration(None=FunctionInput(Preview=@n82.RGBA) {bUsePreviewValueAsDefault=true, inputName=Object-SafeZone-Middle, inputType=FunctionInput_Vector4, sortPriority=200}) {name=SafeZone_Middle, nodeColor=<rgba 0 0.9297 1 1>}
  NamedRerouteDeclaration(None=FunctionInput(Preview=@n82.RGBA) {bUsePreviewValueAsDefault=true, inputName=Object-SafeZone-Front, inputType=FunctionInput_Vector4, sortPriority=200}) {name=SafeZone_Front, nodeColor=<rgba 0 0.9297 1 1>}
  n130 = NamedRerouteUsage()
  NamedRerouteDeclaration(None=Multiply(A=Multiply(A=@n130, B=@n81.Alpha), B=@n81.3V)) {name=CC_Fire_Color, nodeColor=<rgba 0.4766 1 0 1>}
  n131 = VectorParameter() {group=Wrap_MagicMeadow_Fire_Uvs, parameterName=MMead_Fire_A_Scale_Weapon, defaultValue=<rgba 1.5 2 1 1>}
  n132 = VectorParameter() {group=Wrap_MagicMeadow_Fire_Uvs, parameterName=MMead_Fire_A_Scale_Vehicle, defaultValue=<rgba 1 1 1 1>}
  n133 = Multiply(A=@n132.RGB, B=@n132.RGB)
  n134 = VectorParameter() {group=Wrap_MagicMeadow_Fire_Uvs, parameterName=MMead_Fire_A_Scale_VehicleSide, defaultValue=<rgba 1 1 1 1>}
  n135 = Multiply(A=@n134.RGB, B=@n134.RGB)
  n136 = MaterialFunctionCall(Float4=NamedRerouteUsage()) {materialFunction=&BreakOutFloat4Components}
  n137 = VectorParameter() {group=Wrap_MagicMeadow_Fire_Uvs, parameterName=MMead_Fire_A_Scale_Mech, defaultValue=<rgba 1 1 1 1>}
  n138 = VectorParameter() {group=Wrap_MagicMeadow_Fire_Uvs, parameterName=MMead_Fire_A_Scale_Character, defaultValue=<rgba 1 1 1 1>}
  n139 = LinearInterpolate(
   A=LinearInterpolate(A=Multiply(A=@n131.RGB, B=@n131.RGB), B=LinearInterpolate(A=@n133, B=@n135, Alpha=@n136.G), Alpha=@n136.R)
   B=LinearInterpolate(A=Multiply(A=@n137.RGB, B=@n137.RGB), B=Multiply(A=@n138.RGB, B=@n138.RGB), Alpha=@n136.A)
   Alpha=@n136.B
  )
  n140 = VectorParameter() {group=Wrap_MagicMeadow_Fire_Uvs, parameterName=MMead_Fire_A_Scale_Backpack, defaultValue=<rgba 1 1 1 1>}
  n141 = Multiply(A=@n140.RGB, B=@n140.RGB)
  n142 = MaterialFunctionCall(Float2=NamedRerouteUsage()) {materialFunction=&BreakOutFloat2Components}
  n143 = VectorParameter() {group=Wrap_MagicMeadow_Fire_Uvs, parameterName=MMead_Fire_A_Scale_Pickaxe, defaultValue=<rgba 1 1 1 1>}
  NamedRerouteDeclaration(None=LinearInterpolate(A=LinearInterpolate(A=@n139, B=@n141, Alpha=@n142.R), B=Multiply(A=@n143.RGB, B=@n143.RGB), Alpha=@n142.G)) {name=UVs_Fire_A_Scale, nodeColor=<rgba 0 0.7188 1 1>}
  n144 = VectorParameter() {group=Wrap_MagicMeadow_Fire_Uvs, parameterName=MMead_Fire_B_Scale_Weapon, defaultValue=<rgba 1.5 3 1 1>}
  n145 = VectorParameter() {group=Wrap_MagicMeadow_Fire_Uvs, parameterName=MMead_Fire_B_Scale_Vehicle, defaultValue=<rgba 1 1 1 1>}
  n146 = Multiply(A=@n145.RGB, B=@n145.RGB)
  n147 = VectorParameter() {group=Wrap_MagicMeadow_Fire_Uvs, parameterName=MMead_Fire_B_Scale_VehicleSide, defaultValue=<rgba 1 1 1 1>}
  n148 = Multiply(A=@n147.RGB, B=@n147.RGB)
  n149 = MaterialFunctionCall(Float4=NamedRerouteUsage()) {materialFunction=&BreakOutFloat4Components}
  n150 = VectorParameter() {group=Wrap_MagicMeadow_Fire_Uvs, parameterName=MMead_Fire_B_Scale_Mech, defaultValue=<rgba 1 1 1 1>}
  n151 = VectorParameter() {group=Wrap_MagicMeadow_Fire_Uvs, parameterName=MMead_Fire_B_Scale_Character, defaultValue=<rgba 1 1 1 1>}
  n152 = LinearInterpolate(
   A=LinearInterpolate(A=Multiply(A=@n144.RGB, B=@n144.RGB), B=LinearInterpolate(A=@n146, B=@n148, Alpha=@n149.G), Alpha=@n149.R)
   B=LinearInterpolate(A=Multiply(A=@n150.RGB, B=@n150.RGB), B=Multiply(A=@n151.RGB, B=@n151.RGB), Alpha=@n149.A)
   Alpha=@n149.B
  )
  n153 = VectorParameter() {group=Wrap_MagicMeadow_Fire_Uvs, parameterName=MMead_Fire_B_Scale_Backpack, defaultValue=<rgba 1 1 1 1>}
  n154 = Multiply(A=@n153.RGB, B=@n153.RGB)
  n155 = MaterialFunctionCall(Float2=NamedRerouteUsage()) {materialFunction=&BreakOutFloat2Components}
  n156 = VectorParameter() {group=Wrap_MagicMeadow_Fire_Uvs, parameterName=MMead_Fire_B_Scale_Pickaxe, defaultValue=<rgba 1 1 1 1>}
  NamedRerouteDeclaration(None=LinearInterpolate(A=LinearInterpolate(A=@n152, B=@n154, Alpha=@n155.R), B=Multiply(A=@n156.RGB, B=@n156.RGB), Alpha=@n155.G)) {name=UVs_Fire_B_Scale, nodeColor=<rgba 0 0.7188 1 1>}
  NamedRerouteDeclaration(None=@n30) {name=UVs_IconB_Rotation, nodeColor=<rgba 0 0.625 1 1>}
  n157 = PreSkinnedPosition()
  NamedRerouteDeclaration(None=FunctionInput(Preview=VertexInterpolator(VS=@n157).PS) {bUsePreviewValueAsDefault=true, inputName=Pre-Skinned Position, sortPriority=10}) {name=PreSkin_Position, nodeColor=<rgba 0.8281 1 0 1>}
  NamedRerouteDeclaration(None=FunctionInput(Preview=Constant3Vector() {constant=<rgba 0 -23.0279 0 0>}) {bUsePreviewValueAsDefault=true, inputName=Bounds_Max, sortPriority=101}) {name=Bounds_Max, nodeColor=<rgba 0 0.9297 1 1>}
  NamedRerouteDeclaration(None=FunctionInput(Preview=Constant3Vector() {constant=<rgba 0 2.7572 0 0>}) {bUsePreviewValueAsDefault=true, inputName=Bounds_min, sortPriority=102}) {name=Bounds_Min, nodeColor=<rgba 0 0.9297 1 1>}
  n158 = Saturate(None=@n18)
  n159 = CurveAtlasRowParameter(CurveTime=@n158) {
   atlas=&CruveAtlas_Wraps_01
   curve=/Game/Weapons/WeaponSkins/Wraps/Textures/0_ColorCurves/CC_Muzzle_Color
   defaultValue=108
   group=Wrap_MagicMeadow_Fire
   parameterName=MMead__Muzzle_CC
  }
  n160 = MaterialFunctionCall(
   Index=ScalarParameter() {defaultValue=108, group=Wrap_MagicMeadow_Fire, parameterName="MMead__Muzzle_CC  Index"}
   CCinput=@n159
   CCinput_Alpha=@n159
   UV=@n158
  ) {materialFunction=&MF_Global_CC_Tex}
  n161 = Saturate(None=Divide(A=@n19) {constB=0.0001})
  n162 = Divide(A=NamedRerouteUsage(), B=ScalarParameter() {defaultValue=0.25, group=Wrap_MagicMeadow_Reactive, parameterName=MMead_Muzzle_Duration})
  n163 = Multiply(
   A=@n160.Alpha
   B=LinearInterpolate(
    A=ScalarParameter() {defaultValue=1, group=Wrap_MagicMeadow_Reactive, parameterName=MMead_Muzzle_Start_Brightness}
    B=ScalarParameter() {group=Wrap_MagicMeadow_Reactive, parameterName=MMead_Muzzle_End_Brightness}
    Alpha=Saturate(None=@n162)
   )
  )
  n164 = NamedRerouteUsage()
  NamedRerouteDeclaration(None=Multiply(A=Multiply(A=@n160.3V, B=@n161), B=Multiply(A=@n163, B=@n164))) {name=Muzzle_Color, nodeColor=<rgba 0 0.7422 1 1>}
  n165 = NamedRerouteUsage()
  n166 = NamedRerouteUsage()
  n167 = FunctionInput(Preview=@n75.R) {bUsePreviewValueAsDefault=true, inputName=Mask Primary, inputType=FunctionInput_Scalar, sortPriority=24, previewValue=<1 0 0 1>}
  n168 = FunctionInput(Preview=@n75.B) {bUsePreviewValueAsDefault=true, inputName=Mask Secondary, inputType=FunctionInput_Scalar, sortPriority=24, previewValue=<1 0 0 1>}
  n169 = Saturate(None=Add(A=@n167, B=@n168))
  n170 = NamedRerouteUsage()
  n171 = SmoothStep(Value=@n170) {constMax=0.5, constMin=0.2}
  n172 = OneMinus(None=MaterialFunctionCall(Float4=NamedRerouteUsage()) {materialFunction=&BreakOutFloat4Components}.A)
  n173 = Multiply(A=NamedRerouteDeclaration(None=@n168) {name=Msk_Accent, nodeColor=<rgba 0 1 0.7656 1>}, B=Saturate(None=Multiply(A=@n171, B=@n172)))
  NamedRerouteDeclaration(None=BlendMaterialAttributes(A=BlendMaterialAttributes(A=@n165, B=@n166, Alpha=@n169) {vertexAttributeBlendType=UseA}, B=NamedRerouteUsage(), Alpha=Saturate(None=@n173)) {vertexAttributeBlendType=UseA}) {name=Final_Mat, nodeColor=<rgba 1 0.375 0 1>}
  n174 = GetMaterialAttributes(None=NamedRerouteUsage()) {attributeGetTypes [#2, #6, #4]}
  n175 = ScalarParameter() {defaultValue=76, group=Wrap_MagicMeadow_Metal_LUT, parameterName="MMead_Metal_LUT_CC  Index", sortPriority=-5}
  n176 = Saturate(None=DotProduct(A=CameraVectorWS(), B=PixelNormalWS()))
  n177 = CurveAtlasRowParameter(CurveTime=@n176) {
   atlas=&CruveAtlas_Wraps_01
   curve=/Game/Weapons/WeaponSkins/Wraps/Textures/0_ColorCurves/CC_MetalicLUT
   defaultValue=76
   group=Wrap_MagicMeadow_Metal_LUT
   parameterName=MMead_Metal_LUT_CC
   primitiveDataIndex=1
   sortPriority=-5
  }
  n178 = Multiply(
   A=ChannelMaskParameter(None=MaterialFunctionCall(Index=@n175, CCinput=@n177, CCinput_Alpha=@n177, UV=@n176) {materialFunction=&MF_Global_CC_Tex}.4v) {group=Wrap_MagicMeadow_Metal_LUT, parameterName=MMead_Metal_LUT_Channel, sortPriority=-5}.R
   B=@n174.BaseColor
  )
  n179 = MaterialFunctionCall(Float4=NamedRerouteUsage()) {materialFunction=&BreakOutFloat4Components}
  n180 = Max(A=@n179.R, B=@n179.R)
  n181 = Max(A=@n179.B, B=@n179.B)
  n182 = MaterialFunctionCall(Float2=NamedRerouteUsage()) {materialFunction=&BreakOutFloat2Components}
  n183 = LinearInterpolate(A=Add(A=@n174.EmissiveColor, B=NamedRerouteUsage()), B=@n174.EmissiveColor, Alpha=Max(A=Max(A=@n180, B=@n181), B=Max(A=@n182.R, B=@n182.R)))
  n184 = MaterialFunctionCall(Value_Morning=Constant() {r=0.5}, Value_Day=Constant() {r=1}, Value_Evening=Constant() {r=0.5}, Value_Night=Constant() {r=0.25}) {materialFunction=&Scalar_Time_of_Day}
  NamedRerouteDeclaration(None=SetMaterialAttributes!(
    MaterialAttributes=@n174.MaterialAttributes
    BaseColor=LinearInterpolate(A=@n174.BaseColor, B=@n178, Alpha=@n174.BaseColor)
    EmissiveColor=Multiply(A=@n183, B=@n184.Result)
   ) {attributeSetTypes [#2, #6]}) {name=Muzzle_Changes, nodeColor=<rgba 0 1 0.7188 1>}
  n185 = ScalarParameter() {group=Wrap_MagicMeadow_Fire, parameterName=MMead_Fire_Alt_Axies_Contrast}
  n186 = VectorParameter() {group=Wrap_MagicMeadow_Fire, parameterName=MMead_Fire_Alt_Axies_Color}
  n187 = AppendVector(A=@n186.RGB, B=@n186.RGB)
  n188 = Multiply(A=@n10, B=VectorParameter() {group=Wrap_MagicMeadow_Fire_Uvs, parameterName=MMead_Fire_B_Speed, defaultValue=<rgba 0.6 -0.073 0 1>}.RGB)
  n189 = NamedRerouteUsage()
  NamedRerouteDeclaration(None=MaterialFunctionCall!(
    Texture=NamedRerouteUsage()
    Caculated_Bounds=NamedRerouteUsage()
    Pre_Skinned_Normals=NamedRerouteUsage()
    Blend_Contrast=@n185
    IsStatsB=NamedRerouteUsage()
    IsState=NamedRerouteUsage()
    Fill_alternate_Axis_Opacity=Constant() {r=1}
    "Fill_alternate_Axis_Color_A=Roughness"=@n187
    Offset=Add(A=@n188, B=@n189)
    Scale=NamedRerouteUsage()
    Offset_Pivot
    Rotation_Axis=NamedRerouteUsage()
   ) {materialFunction=&/MF_TopsSideProjection_Base_V2}.Combined) {name=TopSide_Flames_B, nodeColor=<rgba 0 0.8828 1 1>}
  NamedRerouteDeclaration(None=MaterialFunctionCall(
    Texture=NamedRerouteUsage()
    Caculated_Bounds=NamedRerouteUsage()
    Pre_Skinned_Normals=NamedRerouteUsage()
    Blend_Contrast=@n185
    IsStatsB=NamedRerouteUsage()
    IsState=NamedRerouteUsage()
    Fill_alternate_Axis_Opacity=Constant() {r=1}
    "Fill_alternate_Axis_Color_A=Roughness"=@n187
    Offset=@n1
    Scale=NamedRerouteUsage()
    Rotation_Axis=NamedRerouteUsage()
   ) {materialFunction=&MF_TopsSideProjection_Base_V2}.Combined) {name=TopSide_Flames_A, nodeColor=<rgba 0 0.8828 1 1>}
  NamedRerouteDeclaration(None=@n40) {name=UVs_Grad_Rotation, nodeColor=<rgba 0 0.625 1 1>}
  n190 = VectorParameter() {group=Wrap_MagicMeadow_Icon_B_Uvs, parameterName=MMead_IconB_Offset_Scale_Weapon, defaultValue=<rgba -0.43 -0.01 0 0.1>}
  n191 = VectorParameter() {group=Wrap_MagicMeadow_Icon_B_Uvs, parameterName=MMead_IconB_Offset_Scale_Vehicle, defaultValue=<rgba 0 0 0 1>}
  n192 = AppendVector(A=@n191.RGB, B=@n191.RGB)
  n193 = VectorParameter() {group=Wrap_MagicMeadow_Icon_B_Uvs, parameterName=MMead_IconB_Offset_Scale_VehicleSide, defaultValue=<rgba 0 0 0 1>}
  n194 = AppendVector(A=@n193.RGB, B=@n193.RGB)
  n195 = MaterialFunctionCall(Float4=NamedRerouteUsage()) {materialFunction=&BreakOutFloat4Components}
  n196 = VectorParameter() {group=Wrap_MagicMeadow_Icon_B_Uvs, parameterName=MMead_IconB_Offset_Scale_Mech, defaultValue=<rgba 0 0 0 1>}
  n197 = VectorParameter() {group=Wrap_MagicMeadow_Icon_B_Uvs, parameterName=MMead_IconB_Offset_Scale_Character, defaultValue=<rgba 0 0 0 1>}
  n198 = LinearInterpolate(
   A=LinearInterpolate(A=AppendVector(A=@n190.RGB, B=@n190.RGB), B=LinearInterpolate(A=@n192, B=@n194, Alpha=@n195.G), Alpha=@n195.R)
   B=LinearInterpolate(A=AppendVector(A=@n196.RGB, B=@n196.RGB), B=AppendVector(A=@n197.RGB, B=@n197.RGB), Alpha=@n195.A)
   Alpha=@n195.B
  )
  n199 = VectorParameter() {group=Wrap_MagicMeadow_Icon_B_Uvs, parameterName=MMead_IconB_Offset_Scale_Backpack, defaultValue=<rgba 0 0 0 1>}
  n200 = AppendVector(A=@n199.RGB, B=@n199.RGB)
  n201 = MaterialFunctionCall(Float2=NamedRerouteUsage()) {materialFunction=&BreakOutFloat2Components}
  n202 = VectorParameter() {group=Wrap_MagicMeadow_Icon_B_Uvs, parameterName=MMead_IconB_Offset_Scale_Pickaxe, defaultValue=<rgba 0 0 0 1>}
  NamedRerouteDeclaration(None=LinearInterpolate(A=LinearInterpolate(A=@n198, B=@n200, Alpha=@n201.R), B=AppendVector(A=@n202.RGB, B=@n202.RGB), Alpha=@n201.G)) {name=UVs_IconB_Offset_Scale, nodeColor=<rgba 0 0.625 1 1>}
  NamedRerouteDeclaration(None=TextureObjectParameter() {group=Wrap_MagicMeadow, parameterName=MMead__tex_Msk3, texture=&T_BlacksRGB}) {name=Tex_Msk3, nodeColor=<rgba 1 0 0.4922 1>}
  NamedRerouteDeclaration(None=@n59) {name=IconA_Msk, nodeColor=<rgba 0.9453 0 1 1>}
  NamedRerouteDeclaration(None=OneMinus(None=@n62.B)) {name=Icon_Accent_Msk, nodeColor=<rgba 1 0.9141 0 1>}
  NamedRerouteDeclaration(None=FunctionInput(Preview=Constant4Vector().RGBA) {bUsePreviewValueAsDefault=true, inputName=IsState, inputType=FunctionInput_Vector4, sortPriority=4}) {name=IsState, nodeColor=<rgba 0.1497 1 0.1633 1>}
  n203 = ScalarParameter() {defaultValue=0.3, group=Wrap_MagicMeadow, parameterName=MMead_Roughness_BckG}
  n204 = ScalarParameter() {defaultValue=0.4, group=Wrap_MagicMeadow, parameterName=MMead_Roughness_Icon_B_Color_A}
  NamedRerouteDeclaration(None=LinearInterpolate(
    A=LinearInterpolate(A=@n203, B=@n204, Alpha=@n62.G)
    B=ScalarParameter() {defaultValue=0.5, group=Wrap_MagicMeadow, parameterName=MMead_Roughness_Icon_B_Color_B}
    Alpha=@n71
   )) {name=Icon_Roughness, nodeColor=<rgba 1 0.9141 0 1>}
  n205 = Subtract(A=Time(), B=ScalarParameter() {defaultValue=-9999999, group=Wrap_MagicMeadow_Reactive, parameterName=MMead_Time_BP})
  NamedRerouteDeclaration(None=@n205) {name=Reactive_Time_From_Bp, nodeColor=<rgba 0.6406 0 1 1>}
  n206 = Saturate(None=Divide(A=@n205, B=ScalarParameter() {defaultValue=0.25, group=Wrap_MagicMeadow_Reactive, parameterName=MMead_Base_Duration}))
  NamedRerouteDeclaration(None=@n206) {name=Reactive_Base_Time, nodeColor=<rgba 0.6406 0 1 1>}
  NamedRerouteDeclaration(None=LinearInterpolate(
    A=ScalarParameter() {defaultValue=3, group=Wrap_MagicMeadow_Reactive, parameterName=MMead_Fire_Start_Brightness}
    B=ScalarParameter() {defaultValue=1, group=Wrap_MagicMeadow_Reactive, parameterName=MMead_Fire_End_Brightness}
    Alpha=@n206
   )) {name=Reactive_Brightness, nodeColor=<rgba 0.4766 1 0 1>}
  n207 = SmoothStep(
   Max=ScalarParameter() {defaultValue=0.6, group=Wrap_MagicMeadow_Icon_A, parameterName=MMead_IconA_Msking Size Accent}
   Value=MaterialFunctionCall(Vector 2=Add(A=@n51.UVs) {constB=-0.5}) {materialFunction=&VectorLength}.V2 Length
  ) {constMax=0, constMin=0.4}
  NamedRerouteDeclaration(None=Saturate(None=Multiply(A=@n207) {constB=10})) {name=IconA_Msk_B, nodeColor=<rgba 0.9453 0 1 1>}
  NamedRerouteDeclaration(None=@n64) {name=Icon_B_Msk, nodeColor=<rgba 1 0.9141 0 1>}
  n208 = ScalarParameter() {defaultValue=1, group=Wrap_MagicMeadow_Fire, parameterName=MMead_Grad_Scale_Weapon}
  n209 = MaterialFunctionCall(Float4=NamedRerouteUsage()) {materialFunction=&BreakOutFloat4Components}
  n210 = LinearInterpolate(
   A=ScalarParameter() {defaultValue=1, group=Wrap_MagicMeadow_Fire, parameterName=MMead_Grad_Scale_Vehicle}
   B=ScalarParameter() {defaultValue=1, group=Wrap_MagicMeadow_Fire, parameterName=MMead_Grad_Scale_Vehicle_Side}
   Alpha=@n209.G
  )
  n211 = ScalarParameter() {defaultValue=1, group=Wrap_MagicMeadow_Fire, parameterName=MMead_Grad_Scale_Mech}
  n212 = ScalarParameter() {defaultValue=1, group=Wrap_MagicMeadow_Fire, parameterName=MMead_Grad_Scale_Character}
  n213 = MaterialFunctionCall(Float2=NamedRerouteUsage()) {materialFunction=&BreakOutFloat2Components}
  n214 = LinearInterpolate(
   A=LinearInterpolate(A=LinearInterpolate(A=@n208, B=@n210, Alpha=@n209.R), B=LinearInterpolate(A=@n211, B=@n212, Alpha=@n209.A), Alpha=@n209.B)
   B=ScalarParameter() {defaultValue=1, group=Wrap_MagicMeadow_Fire, parameterName=MMead_Grad_Scale_Backpack}
   Alpha=@n213.R
  )
  n215 = ScalarParameter() {defaultValue=1, group=Wrap_MagicMeadow_Fire, parameterName=MMead_Grad_Scale_Pickaxe}
  NamedRerouteDeclaration(None=AppendVector(A=LinearInterpolate(A=@n214, B=@n215, Alpha=@n213.G), B=Constant() {r=1})) {name=UVs_Grad_Scale, nodeColor=<rgba 0 0.625 1 1>}
  NamedRerouteDeclaration(None=@n62.G) {name=icons_lines, nodeColor=<rgba 0.7578 0 1 1>}
  NamedRerouteDeclaration(None=@n79) {name=gradent, nodeColor=<rgba 0.8984 0 1 1>}
  NamedRerouteDeclaration(None=FunctionInput(Preview=Constant2Vector()) {bUsePreviewValueAsDefault=true, inputName=IsState_B, inputType=FunctionInput_Vector2, sortPriority=4}) {name=IsState_B, nodeColor=<rgba 0.1497 1 0.1633 1>}
  n216 = Divide(
   A=Subtract(A=Time(), B=ScalarParameter() {defaultValue=-9999999, group=Wrap_MagicMeadow_Reactive, parameterName=MMead_Time_BP_Kill})
   B=ScalarParameter() {defaultValue=0.25, group=Wrap_MagicMeadow_Reactive, parameterName=MMead_Kill_Duration}
  )
  NamedRerouteDeclaration(None=LinearInterpolate(
    A=ScalarParameter() {defaultValue=3, group=Wrap_MagicMeadow_Reactive, parameterName=MMead_Kill_Start_Brightness}
    B=ScalarParameter() {group=Wrap_MagicMeadow_Reactive, parameterName=MMead_Kill_End_Brightness}
    Alpha=Saturate(None=@n216)
   )) {name=Reactive_Brightness_Kill, nodeColor=<rgba 0.4766 1 0 1>}
  NamedRerouteDeclaration(None=@n71) {name=Icon_KillCount_Msk, nodeColor=<rgba 1 0.9141 0 1>}
  n217 = PreSkinnedLocalBounds()
  n218 = Subtract(A=VertexInterpolator(VS=PreSkinnedPosition()).PS, B=@n217.Min)
  n219 = MaterialFunctionCall(Color=Subtract(A=@n217.Max, B=@n217.Max)) {materialFunction=&SplitComponents}
  n220 = Max(A=Max(A=@n219.R, B=@n219.R), B=@n219.B)
  NamedRerouteDeclaration(None=FunctionInput(Preview=Divide(A=@n218, B=@n220)) {bUsePreviewValueAsDefault=true, inputName=Calcated_Bounds, sortPriority=11}) {name=Caclcated_Bounds, nodeColor=<rgba 0 0.9297 1 1>}
  NamedRerouteDeclaration(None=TextureObjectParameter() {group=Wrap_MagicMeadow_Icon_A, parameterName=MMead__tex_Msk, texture=&T_BlacksRGB}) {name=Tex_Msk1, nodeColor=<rgba 1 0 0.4922 1>}
  n221 = PreSkinnedNormal()
  NamedRerouteDeclaration(None=FunctionInput(Preview=VertexInterpolator(VS=@n221).PS) {bUsePreviewValueAsDefault=true, inputName=Pre-Skinned Local Normal, sortPriority=10}) {name=PreSkin_Normal, nodeColor=<rgba 0.8281 1 0 1>}
  NamedRerouteDeclaration(None=@n22.Normal) {name=Root_Normal, nodeColor=<rgba 0.4062 0 1 1>}
  NamedRerouteUsage()
  NamedRerouteUsage()
  NamedRerouteUsage()
  OneMinus(None=@n79)
  OneMinus(None=Add(A=@n17) {constB=0})
  OneMinus(None=@n170)
  OneMinus()
  Saturate(None=Add(A=@n63, B=@n71))
  Saturate(None=@n56)
 ]}
