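# Material function for the Mustang wrap skin on weapons, vehicles and mechs, gated by the Wrap_Mustang static switch.
# isState blends per-target settings for bounds-based projections of the base texture and mouth decal, applied with two customization masks.
# Safe-zone inputs place a ringed badge and a decal revealed in two halves by KillC offsets, apparently a kill counter.
# That decal's emissive shifts between two colors over a duration from a time parameter.
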MaterialFunction /Game/Weapons/WeaponSkins/Wraps/MaterialFunctions/MF_Wrap_Mustang
graph {nodes [
  n1 = FunctionInput(Preview=Constant() {r=1}) {inputName=isState, inputType=FunctionInput_Vector4, sortPriority=5}
  n2 = MaterialFunctionCall(Color=@n1) {materialFunction=&SplitComponents}
  n3 = ScalarParameter() {group=Wrap_Mustang, parameterName=Mustang_BC_Decals_Rotation_Vehicle}
  n4 = ScalarParameter() {group=Wrap_Mustang, parameterName=Mustang_BC_Decals_Rotation_VehicleSide}
  n5 = MaterialFunctionCall(Color=@n1) {materialFunction=&SplitComponents}
  n6 = LinearInterpolate(
   A=LinearInterpolate(
    A=ScalarParameter() {group=Wrap_Mustang, parameterName=Mustang_BC_Decals_Rotation_Weapon}
    B=LinearInterpolate(A=@n3, B=@n4, Alpha=@n5.G)
    Alpha=@n5.R
   )
   B=ScalarParameter() {group=Wrap_Mustang, parameterName=Mustang_BC_Decals_Rotation_Mech}
   Alpha=@n5.B
  )
  n7 = Constant4Vector() {constant=<rgba 0 0 0 0.5>}
  n8 = FunctionInput(Preview=@n7.RGBA) {bUsePreviewValueAsDefault=true, inputName=Object-SafeZone-Middle, inputType=FunctionInput_Vector4, sortPriority=101}
  n9 = VertexInterpolator(VS=PreSkinnedPosition())
  n10 = PreSkinnedLocalBounds()
  n11 = FunctionInput(Preview=Divide(A=Subtract(A=@n9.PS, B=@n10.Min), B=Subtract(A=@n10.Max, B=@n10.Max))) {bUsePreviewValueAsDefault=true, inputName=Calcated_Bounds, sortPriority=11}
  n12 = FunctionInput(Preview=VertexInterpolator(VS=PreSkinnedNormal()).PS) {bUsePreviewValueAsDefault=true, inputName=Pre-Skinned Local Normal, sortPriority=10}
  n13 = MaterialFunctionCall!(
   For_Safe_Zones=StaticBool() {value=true}
   Normal_Map
   Caculated_Bounds=@n11
   Pre_Skinned_Normals=@n12
   Is Vehicle=@n2.R
   Is VehicleSide=@n2.R
   Is Mech=@n2.R
   Vehicle_Rotation
   UV_Rotation Bypass=@n6
   Mech_Rotation
   Uvs_Offset=ComponentMask(None=@n8) {b=true, g=true, r=true}
   Offset-after rotation
   Uvs_TileRate=ComponentMask(None=@n8) {a=true}
  ) {materialFunction=&/MF_Projection_Side_Normal}
  n14 = ComponentMask(None=@n13.UVs) {g=true}
  n15 = Floor(None=Multiply(A=@n14) {constB=4})
  n16 = Multiply(A=Add(A=@n14) {constB=-0.25}) {constB=4}
  Add(A=Saturate(None=OneMinus(None=@n15)), B=Saturate(None=Floor(None=@n16)))
  n17 = MaterialFunctionCall(Color=Abs(None=@n12)) {materialFunction=&SplitComponents}
  n18 = MaterialFunctionCall(Color=@n1) {materialFunction=&SplitComponents}
  n19 = LinearInterpolate(A=@n17.B, B=@n17.B, Alpha=@n18.G)
  n20 = MaterialFunctionCall(
   In=LinearInterpolate(A=LinearInterpolate(A=@n17.R, B=@n19, Alpha=@n18.R), B=@n17.G, Alpha=@n18.B)
   Contrast=ScalarParameter() {defaultValue=1, group=Wrap_Mustang, parameterName=Mustang_TriPlanner_Mask_Contrast, sortPriority=-5}
  ) {materialFunction=&CheapContrast}
  Add(A=ScalarParameter() {defaultValue=1, group=Wrap_Mustang, parameterName="Mustang_TriPlanner_Mask 0=On 1=Off", sortPriority=-5}, B=@n20.Result)
  n21 = FunctionInput(Preview=SetMaterialAttributes!(MaterialAttributes, Normal) {attributeSetTypes [#1]}) {bUsePreviewValueAsDefault=true, inputName=Material, inputType=FunctionInput_MaterialAttributes, sortPriority=-20}
  n22 = GetMaterialAttributes(None=@n21) {attributeGetTypes [
    #2
    #3
    #4
    #5
   ]}
  n23 = TextureSample() {samplerType=SAMPLERTYPE_Masks, texture=/Game/Weapons/FORT_Rifles/Textures/T_Scar_CM}
  n24 = FunctionInput(Preview=@n23.R) {bUsePreviewValueAsDefault=true, inputName=Customization Mask A, inputType=FunctionInput_Scalar, sortPriority=24}
  n25 = VectorParameter() {group=Wrap_Mustang, parameterName=Mustang_BaseColor, sortPriority=0, defaultValue=<rgba 0.003 0.003 0.003 1>}
  n26 = MaterialFunctionCall(Color=@n1) {materialFunction=&SplitComponents}
  n27 = MaterialFunctionCall(Color=@n1) {materialFunction=&SplitComponents}
  n28 = LinearInterpolate(
   A=ScalarParameter() {group=Wrap_Mustang, parameterName=Mustang_BC_Rotation_Weapon}
   B=LinearInterpolate(
    A=ScalarParameter() {group=Wrap_Mustang, parameterName=Mustang_BC_Rotation_Vehicle}
    B=ScalarParameter() {group=Wrap_Mustang, parameterName=Mustang_BC_Rotation_VehicleSide}
    Alpha=@n27.G
   )
   Alpha=@n27.R
  )
  n29 = ScalarParameter() {defaultValue=0.25, group=Wrap_Mustang, parameterName=Mustang_BC_Rotation_Mech}
  n30 = MaterialFunctionCall(Color=@n1) {materialFunction=&SplitComponents}
  n31 = LinearInterpolate(
   A=VectorParameter() {group=Wrap_Mustang, parameterName=Mustang_BC_Offset_Weapon, defaultValue=<rgba 0 0 0 1>}.RGB
   B=LinearInterpolate(
    A=VectorParameter() {group=Wrap_Mustang, parameterName=Mustang_BC_Offset_Vehicle, defaultValue=<rgba 0 0 0 1>}.RGB
    B=VectorParameter() {group=Wrap_Mustang, parameterName=Mustang_BC_Offset_VehicleSide, defaultValue=<rgba 0 0 0 1>}.RGB
    Alpha=@n30.G
   )
   Alpha=@n30.R
  )
  n32 = VectorParameter() {group=Wrap_Mustang, parameterName=Mustang_BC_Offset_Mech, defaultValue=<rgba 0 0 0 1>}
  n33 = MaterialFunctionCall(Color=@n1) {materialFunction=&SplitComponents}
  n34 = LinearInterpolate(
   A=ScalarParameter() {defaultValue=1, group=Wrap_Mustang, parameterName=Mustang_BC_Scale_Weapon}
   B=LinearInterpolate(
    A=ScalarParameter() {defaultValue=1, group=Wrap_Mustang, parameterName=Mustang_BC_Scale_Vehicle}
    B=ScalarParameter() {defaultValue=1, group=Wrap_Mustang, parameterName=Mustang_BC_Scale_VehicleSide}
    Alpha=@n33.G
   )
   Alpha=@n33.R
  )
  n35 = ScalarParameter() {defaultValue=1, group=Wrap_Mustang, parameterName=Mustang_BC_Scale_Mech}
  n36 = TextureSample(
   UVs=MaterialFunctionCall(
    Caculated_Bounds=@n11
    Pre_Skinned_Normals=@n12
    Is Vehicle=@n26.R
    Is VehicleSide=@n26.R
    Is Mech=@n26.R
    UV_Rotation Bypass=LinearInterpolate(A=@n28, B=@n29, Alpha=@n27.B)
    Uvs_Offset=LinearInterpolate(A=@n31, B=@n32.RGB, Alpha=@n30.B)
    Uvs_TileRate=LinearInterpolate(A=@n34, B=@n35, Alpha=@n33.B)
   ) {materialFunction=&MF_Projection_Side_Normal}.UVs
   Tex=TextureObjectParameter() {group=Wrap_Mustang, parameterName=Mustang_BC__Tex, texture=&T_BlacksRGB}
  )
  n37 = VectorParameter() {group=Wrap_Mustang, parameterName=Mustang_2nd_BaseColor, sortPriority=0, defaultValue=<rgba 0.8021 0.5005 0.0033 1>}
  n38 = FunctionInput(Preview=@n23.B) {bUsePreviewValueAsDefault=true, inputName=Customization Mask B, inputType=FunctionInput_Scalar, sortPriority=24}
  n39 = LinearInterpolate(A=LinearInterpolate(A=@n22.BaseColor, B=LinearInterpolate(A=@n25.RGB, B=@n36.RGB, Alpha=@n36.RGB), Alpha=@n24), B=@n37.RGB, Alpha=@n38)
  n40 = VectorParameter() {group=Wrap_Mustang, parameterName=Mustang_Decal_Color, sortPriority=0, defaultValue=<rgba 1 1 1 1>}
  n41 = MaterialFunctionCall(Color=@n1) {materialFunction=&SplitComponents}
  n42 = FunctionInput(Preview=@n7.RGBA) {bUsePreviewValueAsDefault=true, inputName=Object-SafeZone-Back, inputType=FunctionInput_Vector4, sortPriority=100}
  n43 = MaterialFunctionCall(
   For_Safe_Zones=StaticBool() {value=true}
   Caculated_Bounds=@n11
   Pre_Skinned_Normals=@n12
   Is Vehicle=@n41.R
   Is VehicleSide=@n41.R
   Is Mech=@n41.R
   UV_Rotation Bypass=@n6
   Uvs_Offset=ComponentMask(None=@n42) {b=true, g=true, r=true}
   Uvs_TileRate=ComponentMask(None=@n42) {a=true}
  ) {materialFunction=&MF_Projection_Side_Normal}
  n44 = SphereMask(A=@n43.UVs, B=Constant() {r=0.5}, Radius=Constant() {r=0.4}, Hardness=Constant() {r=1})
  n45 = Subtract(A=SphereMask(A=@n43.UVs, B=Constant() {r=0.5}, Radius=Constant() {r=0.42}, Hardness=Constant() {r=1}), B=@n44)
  n46 = MaterialFunctionCall(UVs=@n43.UVs, Texture Scale=Multiply() {constA=1, constB=0.8}) {materialFunction=&ScaleUVsByCenter}
  n47 = TextureObjectParameter() {group=Wrap_Mustang, parameterName=Mustang_Decal_Ban__M_Tex, texture=&T_BlacksRGB}
  n48 = Multiply(A=@n44, B=OneMinus(None=TextureSample(UVs=@n46.UVs, Tex=@n47).R))
  n49 = LinearInterpolate(A=@n25.RGB, B=@n40.RGB, Alpha=Saturate(None=Add(A=@n45, B=@n48)))
  n50 = Saturate(None=Add(A=@n38, B=@n24))
  n51 = Add(A=OneMinus(None=ComponentMask(None=@n13.UVs) {r=true})) {constB=-1}
  n52 = Divide(A=ScalarParameter() {group=Wrap_Mustang, parameterName=Mustang_KillC_Offset_A}) {constA=10, constB=8}
  n53 = Saturate(None=Divide(A=Add(A=@n51, B=@n52)) {constB=0})
  n54 = ComponentMask(None=@n13.UVs) {g=true}
  n55 = Saturate(None=Round(None=OneMinus(None=@n54)))
  n56 = Saturate(None=Multiply(A=@n53, B=@n55))
  n57 = OneMinus(None=ComponentMask(None=@n13.UVs) {r=true})
  n58 = ScalarParameter() {group=Wrap_Mustang, parameterName=Mustang_KillC_Offset_B}
  n59 = Divide(A=Add(A=Add(A=@n57) {constB=-1}, B=Divide(A=@n58) {constA=10, constB=8})) {constB=0}
  n60 = Saturate(None=Subtract(A=Saturate(None=@n59), B=@n55))
  n61 = TextureObjectParameter() {group=Wrap_Mustang, parameterName=Mustang_Decal__M_Tex, texture=&T_BlacksRGB}
  n62 = ComponentMask(None=@n13.UVs) {r=true}
  n63 = Constant() {r=0.5}
  n64 = ComponentMask(None=@n13.UVs) {g=true}
  n65 = Multiply(
   A=Multiply(A=Add(A=@n56, B=@n60), B=TextureSample(UVs=@n13.UVs, Tex=@n61).RGBA)
   B=Multiply(A=SphereMask(A=@n62, B=@n63, Radius=@n63), B=SphereMask(A=@n64, B=@n63, Radius=@n63))
  )
  n66 = SphereMask(A=@n43.UVs, B=Constant() {r=0.5}, Radius=Constant() {r=0.49}, Hardness=Constant() {r=1})
  n67 = Add(A=@n66, B=ComponentMask(None=@n65) {b=true})
  n68 = MaterialFunctionCall!(
   Normal_Map
   Caculated_Bounds=@n11
   Pre_Skinned_Normals=@n12
   Is Vehicle=@n2.R
   Is Vehicle side=@n2.R
   Is Mech=@n2.R
   Offset-after rotation
   UV_Rotation_Right
   UV_Rotation_Left
   Uvs_TileRate_Right
   Uvs_TileRate_Left
   Uvs_Offset_Right
   Uvs_Offset_Left
  ) {materialFunction=&/MF_Projection_Side_Left_Right_Decal_Normal}
  n69 = MaterialFunctionCall(Color=@n1) {materialFunction=&SplitComponents}
  n70 = LinearInterpolate(A=LinearInterpolate(B=LinearInterpolate(A=@n68.Left_Right_Mask, Alpha=@n69.G), Alpha=@n69.R) {constA=1}, Alpha=@n69.B)
  n71 = Multiply(A=Saturate(None=Multiply(A=@n67, B=@n70)), B=@n50)
  n72 = MaterialFunctionCall(Color=Multiply(A=@n11, B=Constant3Vector() {constant=<rgba 1 1 -1 0>})) {materialFunction=&SplitComponents}
  n73 = AppendVector(A=@n72.G, B=@n72.G)
  n74 = AppendVector(A=@n72.R, B=@n72.R)
  n75 = MaterialFunctionCall(Float4=@n1) {materialFunction=&BreakOutFloat4Components}
  n76 = LinearInterpolate(A=@n73, B=@n74, Alpha=@n75.R)
  n77 = LinearInterpolate(A=@n73, B=@n74, Alpha=@n75.A)
  n78 = MaterialFunctionCall(Float4=@n1) {materialFunction=&BreakOutFloat4Components}
  n79 = LinearInterpolate(
   A=VectorParameter() {group=Wrap_Mustang, parameterName=Mustang_DecalMouth_Offset_Weapon, defaultValue=<rgba 0 1 0 1>}.RGB
   B=LinearInterpolate(
    A=VectorParameter() {group=Wrap_Mustang, parameterName=Mustang_DecalMouth_Offset_Vehicle, defaultValue=<rgba 0 0 0 1>}.RGB
    B=VectorParameter() {group=Wrap_Mustang, parameterName=Mustang_DecalMouth_Offset_VehicleSide, defaultValue=<rgba 0 0 0 1>}.RGB
    Alpha=@n78.G
   )
   Alpha=@n78.R
  )
  n80 = LinearInterpolate(
   A=VectorParameter() {group=Wrap_Mustang, parameterName=Mustang_DecalMouth_Offset_Mech}.RGB
   B=VectorParameter() {group=Wrap_Mustang, parameterName=Mustang_DecalMouth_Offset_Character}.RGB
   Alpha=@n78.A
  )
  n81 = Add(
   A=LinearInterpolate(
    A=LinearInterpolate(A=@n76, B=@n77, Alpha=@n75.B)
    B=AppendVector(A=@n72.R, B=@n72.R)
    Alpha=ScalarParameter() {group=Wrap_Mustang, parameterName=Mustang_Is_Roll, sortPriority=0}
   )
   B=ComponentMask(None=LinearInterpolate(A=@n79, B=@n80, Alpha=@n78.B)) {g=true, r=true}
  ) {constB=0}
  n82 = Constant2Vector() {g=0.5, r=0.5}
  n83 = ScalarParameter() {group=Wrap_Mustang, parameterName=Mustang_DecalMouth_Rotation_Vehicle}
  n84 = ScalarParameter() {group=Wrap_Mustang, parameterName=Mustang_DecalMouth_Rotation_VehicleSide}
  n85 = MaterialFunctionCall(Float4=@n1) {materialFunction=&BreakOutFloat4Components}
  n86 = LinearInterpolate(
   A=LinearInterpolate(
    A=ScalarParameter() {group=Wrap_Mustang, parameterName=Mustang_DecalMouth_Rotation_Weapon}
    B=LinearInterpolate(A=@n83, B=@n84, Alpha=@n85.G)
    Alpha=@n85.R
   )
   B=LinearInterpolate(
    A=ScalarParameter() {defaultValue=0.25, group=Wrap_Mustang, parameterName=Mustang_DecalMouth_Rotation_Mech}
    B=ScalarParameter() {defaultValue=0.25, group=Wrap_Mustang, parameterName=Mustang_DecalMouth_Rotation_Character}
    Alpha=@n85.A
   )
   Alpha=@n85.B
  )
  n87 = VectorParameter() {group=Wrap_Mustang, parameterName=Mustang_DecalMouth_Scale_Vehicle, defaultValue=<rgba 1 1 0 0.4>}
  n88 = Multiply(A=@n87.RGB, B=@n87.RGB)
  n89 = MaterialFunctionCall(Color=@n1) {materialFunction=&SplitComponents}
  n90 = LinearInterpolate(
   A=ScalarParameter() {defaultValue=0.5, group=Wrap_Mustang, parameterName=Mustang_DecalMouth_Scale_Weapon}
   B=LinearInterpolate(
    A=ComponentMask(None=@n88) {g=true, r=true}
    B=ScalarParameter() {defaultValue=1, group=Wrap_Mustang, parameterName=Mustang_DecalMouth_Scale_VehicleSide}
    Alpha=@n89.G
   )
   Alpha=@n89.R
  )
  n91 = VectorParameter() {group=Wrap_Mustang, parameterName=Mustang_DecalMouth_Scale_MechB, defaultValue=<rgba 1 1 0 0.4>}
  n92 = VectorParameter() {group=Wrap_Mustang, parameterName=Mustang_DecalMouth_Scale_Character, defaultValue=<rgba 1 1 0 0.4>}
  n93 = LinearInterpolate(
   A=ComponentMask(None=Multiply(A=@n91.RGB, B=@n91.RGB)) {g=true, r=true}
   B=ComponentMask(None=Multiply(A=@n92.RGB, B=@n92.RGB)) {g=true, r=true}
   Alpha=@n78.A
  )
  n94 = TextureSample(
   UVs=MaterialFunctionCall(
    UVs=MaterialFunctionCall(UVs=@n81, Rotation Center=@n82, "Rotation Angle (0-1)"=@n86) {materialFunction=&CustomRotator}.Rotated Values
    Texture Scale=LinearInterpolate(A=@n90, B=@n93, Alpha=@n89.B)
   ) {materialFunction=&ScaleUVsByCenter}.UVs
   Tex=TextureObjectParameter() {group=Wrap_Mustang, parameterName=Mustang_DecalMouth__Tex, texture=&T_BlacksRGB}
  )
  n95 = Multiply(A=@n94.A, B=@n50)
  n96 = ScalarParameter() {group=Wrap_Mustang, parameterName=Mustang_Metallic_A, sortPriority=0}
  n97 = ScalarParameter() {group=Wrap_Mustang, parameterName=Mustang_Metallic_B, sortPriority=0}
  n98 = Saturate(None=Saturate(None=Subtract(A=@n36.A, B=@n66)))
  n99 = LinearInterpolate(
   A=LinearInterpolate(A=@n22.Metallic, B=LinearInterpolate(A=@n96, B=@n97, Alpha=@n98), Alpha=@n24)
   B=ScalarParameter() {group=Wrap_Mustang, parameterName=Mustang_2nd_Metallic, sortPriority=0}
   Alpha=@n38
  )
  n100 = ScalarParameter() {group=Wrap_Mustang, parameterName=Mustang_Roghness_A, sortPriority=0}
  n101 = ScalarParameter() {defaultValue=1, group=Wrap_Mustang, parameterName=Mustang_Roghness_B, sortPriority=0}
  n102 = LinearInterpolate(
   A=LinearInterpolate(A=@n22.Roughness, B=LinearInterpolate(A=@n100, B=@n101, Alpha=@n98), Alpha=@n24)
   B=ScalarParameter() {group=Wrap_Mustang, parameterName=Mustang_2nd_Roghness, sortPriority=0}
   Alpha=@n38
  )
  n103 = ScalarParameter() {defaultValue=0.5, group=Wrap_Mustang, parameterName=Mustang_Specular, sortPriority=0}
  n104 = LinearInterpolate(
   A=LinearInterpolate(A=@n22.Specular, B=@n103, Alpha=@n24)
   B=ScalarParameter() {defaultValue=0.5, group=Wrap_Mustang, parameterName=Mustang_2nd_Specular, sortPriority=0}
   Alpha=@n38
  )
  n105 = ComponentMask(None=@n65) {a=true}
  n106 = VectorParameter() {group=Wrap_Mustang, parameterName=Mustang_Time_BP_Start_Color, defaultValue=<rgba 3 2.2297 0 1>}
  n107 = VectorParameter() {group=Wrap_Mustang, parameterName=Mustang_Time_BP_Start_End, defaultValue=<rgba 0 0 0 1>}
  n108 = Time()
  n109 = ScalarParameter() {defaultValue=-9999999, group=Wrap_Mustang, parameterName=Mustang_Time_BP, sortPriority=200}
  n110 = Saturate(None=Divide(
    A=Subtract(A=@n108, B=@n109)
    B=ScalarParameter() {defaultValue=1, group=Wrap_Mustang, parameterName=Mustang_Time_BP_Duration, sortPriority=200}
   ))
  n111 = SetMaterialAttributes!(
   MaterialAttributes=@n22.MaterialAttributes
   BaseColor=LinearInterpolate(A=LinearInterpolate(A=@n39, B=@n49, Alpha=@n71), B=@n94.RGB, Alpha=@n95)
   Metallic=LinearInterpolate(
    A=LinearInterpolate(A=@n99, B=@n96, Alpha=@n71)
    B=ScalarParameter() {group=Wrap_Mustang, parameterName=Mustang_DecalMouth_Metallic, sortPriority=0}
    Alpha=@n95
   )
   Roughness=LinearInterpolate(
    A=LinearInterpolate(A=@n102, B=@n100, Alpha=@n71)
    B=ScalarParameter() {group=Wrap_Mustang, parameterName=Mustang_DecalMouth_Roghness, sortPriority=0}
    Alpha=@n95
   )
   Specular=LinearInterpolate(
    A=LinearInterpolate(A=@n104, B=@n103, Alpha=@n71)
    B=ScalarParameter() {defaultValue=0.5, group=Wrap_Mustang, parameterName=Mustang_DecalMouth_Specular, sortPriority=0}
    Alpha=@n95
   )
   EmissiveColor=Multiply(A=Multiply(A=@n105, B=@n70), B=LinearInterpolate(A=@n106.RGB, B=@n107.RGB, Alpha=@n110))
  ) {attributeSetTypes [
    #2
    #3
    #4
    #5
    #6
   ]}
  BlendMaterialAttributes(A=@n21, B=@n111, Alpha=Saturate(None=@n23.R)) {vertexAttributeBlendType=UseA}
  FunctionInput(Preview=TextureSample() {samplerType=SAMPLERTYPE_Masks, texture=&T_Scar_CM}.G) {bUsePreviewValueAsDefault=true, inputName=Curavture, inputType=FunctionInput_Scalar, sortPriority=24}
  FunctionInput(Preview=Constant3Vector() {constant=<rgba 0 1 0 0>}) {bUsePreviewValueAsDefault=true, inputName=ForwardAxis}
  FunctionInput(Preview=@n7.RGBA) {bUsePreviewValueAsDefault=true, inputName=Object-SafeZone-Front, inputType=FunctionInput_Vector4, sortPriority=102}
  FunctionOutput(None=StaticSwitchParameter(True=@n111, False=@n21) {group=Customization_Wrap, parameterName=Wrap_Mustang})
  Multiply(A=Saturate(None=@n20.Result), B=@n66)
  Power(Base=@n110, Exp=ScalarParameter() {defaultValue=0.25, group=Wrap_Mustang, parameterName=Mustang_Time_BP_Falloff, sortPriority=200})
  Subtract(A=@n53, B=@n15)
 ]}
