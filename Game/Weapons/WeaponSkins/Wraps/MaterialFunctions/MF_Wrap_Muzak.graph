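# Material function for the Muzak weapon wrap, layering an audio-reactive look of warped lines, music notes and an equalizer onto input material attributes.
# Pre-skinned bounds supply UVs switched by weapon, vehicle, side and mech inputs, masked by local normals.
# Event spectrum and a Blueprint-set timer drive line warp, while emote or event audio crossfades a scrolling wave texture into custom-HLSL bars.
# Curve atlases color layers, blended by customization mask behind a static switch.
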MaterialFunction /Game/Weapons/WeaponSkins/Wraps/MaterialFunctions/MF_Wrap_Muzak
graph {nodes [
  n1 = Constant()
  n2 = Multiply(A=Time(), B=AppendVector(A=ScalarParameter() {group=Wrap_Muzak, parameterName=Muz_Line_warp_Speed}, B=@n1))
  n3 = AppendVector(A=ScalarParameter() {group=Wrap_Muzak, parameterName=Muz_Line_warp_Offset}, B=@n1)
  n4 = AppendVector(A=ScalarParameter() {defaultValue=1, group=Wrap_Muzak, parameterName=Muz_Line_warp_Tile}, B=Constant() {r=1})
  n5 = PreSkinnedPosition()
  n6 = PreSkinnedLocalBounds()
  n7 = Divide(A=Subtract(A=VertexInterpolator(VS=@n5).PS, B=@n6.Min), B=Subtract(A=@n6.Max, B=@n6.Max))
  n8 = Multiply(
   A=Add(
    A=FunctionInput(Preview=@n7) {bUsePreviewValueAsDefault=true, inputName=Calcated_Bounds, sortPriority=11}
    B=VectorParameter() {group=Wrap_Muzak, parameterName=Muz_Roll_offset, sortPriority=700, defaultValue=<rgba 0 0 0 1>}.RGB
   )
   B=VectorParameter() {group=Wrap_Muzak, parameterName=Muz_Roll_Scale, sortPriority=700, defaultValue=<rgba 1 1 1 1>}.RGB
  )
  n9 = MaterialFunctionCall(Color=Multiply(A=@n8, B=Constant3Vector() {constant=<rgba 1 1 -1 0>})) {materialFunction=&SplitComponents}
  n10 = ScalarParameter() {group=Wrap_Muzak, parameterName=Muz_Rotation_Vehicle, sortPriority=700}
  n11 = MaterialFunctionCall(UVs=AppendVector(A=@n9.R, B=@n9.R), "Rotation Angle (0-1)"=@n10) {materialFunction=&CustomRotator}
  n12 = AppendVector(A=@n9.R, B=@n9.R)
  n13 = Constant()
  n14 = FunctionInput(Preview=@n13) {bUsePreviewValueAsDefault=true, inputName=Side Project Vehicle, inputType=FunctionInput_Scalar}
  n15 = FunctionInput(Preview=@n13) {bUsePreviewValueAsDefault=true, inputName=isVehicle, inputType=FunctionInput_Scalar, sortPriority=5}
  n16 = FunctionInput(Preview=@n13) {bUsePreviewValueAsDefault=true, inputName=isMech, inputType=FunctionInput_Scalar, sortPriority=4}
  n17 = LinearInterpolate(
   A=LinearInterpolate(A=AppendVector(A=@n9.G, B=@n9.G), B=LinearInterpolate(A=@n11.Rotated Values, B=@n12, Alpha=@n14), Alpha=@n15)
   B=AppendVector(A=@n9.B, B=@n9.B)
   Alpha=@n16
  )
  n18 = TextureObject() {samplerType=SAMPLERTYPE_LinearColor, texture=/Game/Weapons/WeaponSkins/Wraps/Textures/Muzak/T_Wrap_Muzak_M2_1}
  n19 = TextureSample(UVs=Add(A=Add(A=@n2, B=@n3), B=Multiply(A=@n4, B=@n17)), Tex=@n18) {samplerType=SAMPLERTYPE_Masks, texture=/Game/Weapons/WeaponSkins/Wraps/Textures/Muzak/T_Wrap_Muzak_m}
  n20 = Subtract(A=Time(), B=ScalarParameter() {defaultValue=-9999999, group=Wrap_Muzak, parameterName=TimeFrom-BP----------------, sortPriority=600})
  n21 = ScalarParameter() {defaultValue=1.2, group=Wrap_Muzak, parameterName=Muz_Time_Duration, sortPriority=601}
  n22 = Power(
   Base=Saturate(None=Divide(A=@n20, B=@n21))
   Exp=ScalarParameter() {defaultValue=0.5, group=Wrap_Muzak, parameterName=Muz_Time_Cuve, sortPriority=602}
  )
  n23 = CollectionParameter() {collection=&FortniteMaterialParameters, parameterName=EventAudio_Spectrum}
  n24 = MaterialFunctionCall(Float4=@n23) {materialFunction=&BreakOutFloat4Components}
  n25 = Add(A=Multiply(A=@n24.R) {constB=0.75}, B=Multiply(A=@n24.B) {constB=0.75})
  n26 = Max(
   A=OneMinus(None=@n22)
   B=Clamp(None=Power(Base=@n25), Max=ScalarParameter() {defaultValue=0.5, group=Wrap_Muzak, parameterName=Muz_Line_Audio_Enflunce}) {maxDefault=0.5}
  )
  n27 = ScalarParameter() {defaultValue=0.5, group=Wrap_Muzak, parameterName=Muz_Line_Base_Warp_Amount}
  n28 = ScalarParameter() {group=Wrap_Muzak, parameterName=Muz_Line_Mask_Offset}
  n29 = ScalarParameter() {group=Wrap_Muzak, parameterName=Muz_Line_Mask_Offset_Mech}
  n30 = MaterialFunctionCall(Color=Add(A=@n8, B=LinearInterpolate(A=@n28, B=@n29, Alpha=@n16)) {constB=0}) {materialFunction=&SplitComponents}
  n31 = ScalarParameter() {group=Wrap_Muzak, parameterName=Muz_Is_Roll, sortPriority=700}
  n32 = LinearInterpolate(A=LinearInterpolate(A=@n30.G, B=LinearInterpolate(A=@n30.R, B=@n30.R, Alpha=@n31), Alpha=@n15), B=@n30.B, Alpha=@n16)
  n33 = Multiply(A=@n32, B=OneMinus(None=@n32))
  n34 = ScalarParameter() {defaultValue=2, group=Wrap_Muzak, parameterName=Muz_Line_Mask_Power}
  n35 = Saturate(None=Multiply(A=Power(Base=@n33, Exp=@n34), B=ScalarParameter() {defaultValue=1.5, group=Wrap_Muzak, parameterName=Muz_Line_Mask_Mul}))
  n36 = ScalarParameter() {group=Wrap_Muzak, parameterName=Muz_Line_Mask_cotrast}
  n37 = Multiply(A=Saturate(None=Add(A=@n35, B=@n36)), B=Multiply(A=Add(A=@n19.A) {constB=-0.5}, B=Max(A=@n26, B=@n27)))
  n38 = TextureObject() {samplerType=SAMPLERTYPE_Masks, texture=/Game/Weapons/WeaponSkins/Wraps/Textures/Muzak/T_Wrap_Muzak_M2}
  n39 = TextureSample(UVs=Add(A=NamedRerouteDeclaration(None=@n37) {name=Warp_uvs, nodeColor=<rgba 0.7812 1 0 1>}, B=@n17), Tex=@n38) {samplerType=SAMPLERTYPE_Masks, texture=&T_Wrap_Muzak_m}
  Add(A=@n39.A) {constB=0}
  n40 = NamedRerouteUsage()
  Add(
   A=@n40
   B=ComponentMask(None=VectorParameter() {group=Wrap_Muzak_Switch_Line, parameterName=Muz_Switch_Line_Offset, defaultValue=<rgba 0 0 0 1>}.RGB) {g=true, r=true}
  )
  Constant()
  n41 = ScalarParameter() {defaultValue=27, group=Wrap_Muzak, parameterName="Muz_Note_ColorCurve  Index"}
  n42 = MaterialFunctionCall(Color=MaterialFunctionCall(X=@n15, Y=@n14, Z=@n16) {materialFunction=&MakeFloat3}.Result) {materialFunction=&SplitComponents}
  n43 = LinearInterpolate(
   A=VectorParameter() {group=Wrap_Muzak, parameterName=Muz_Note_position_Weapon, defaultValue=<rgba 0 0 0 1>}.RGB
   B=LinearInterpolate(
    A=VectorParameter() {group=Wrap_Muzak, parameterName=Muz_Note_position_Vehicle, defaultValue=<rgba 0 0 0 1>}.RGB
    B=VectorParameter() {group=Wrap_Muzak, parameterName=Muz_Note_position_Vehicle_Side, defaultValue=<rgba 0 0 0 1>}.RGB
    Alpha=@n42.G
   )
   Alpha=@n42.R
  )
  n44 = VectorParameter() {group=Wrap_Muzak, parameterName=Muz_Note_position_Mech, defaultValue=<rgba 0 0 0 1>}
  n45 = Constant2Vector() {g=1}
  n46 = LinearInterpolate(A=@n45, B=LinearInterpolate(A=Constant2Vector(), B=@n45, Alpha=@n42.G), Alpha=@n42.R)
  n47 = Constant2Vector() {r=1}
  n48 = Add(
   A=ComponentMask(None=LinearInterpolate(A=@n43, B=@n44.RGB, Alpha=@n42.B)) {g=true, r=true}
   B=Add(A=LinearInterpolate(A=@n46, B=@n47, Alpha=@n42.B), B=@n17)
  )
  n49 = ScalarParameter() {defaultValue=1, group=Wrap_Muzak, parameterName=Muz_Note_Scale_Vehicle}
  n50 = ScalarParameter() {defaultValue=1, group=Wrap_Muzak, parameterName=Muz_Note_Scale_Vehicle_Side}
  n51 = LinearInterpolate(
   A=LinearInterpolate(
    A=ScalarParameter() {defaultValue=1, group=Wrap_Muzak, parameterName=Muz_Note_Scale_Weapon}
    B=LinearInterpolate(A=@n49, B=@n50, Alpha=@n42.G)
    Alpha=@n42.R
   )
   B=ScalarParameter() {defaultValue=1, group=Wrap_Muzak, parameterName=Muz_Note_Scale_Mech}
   Alpha=@n42.B
  )
  n52 = CollectionParameter() {collection=&FortniteMaterialParameters, parameterName=Emote_AudioLow}
  n53 = CollectionParameter() {collection=&FortniteMaterialParameters, parameterName=Emote_AudioHigh}
  n54 = Add(A=@n52, B=@n53)
  n55 = ScalarParameter() {defaultValue=20, group=Wrap_Muzak, parameterName=Muz_Emote_Mask}
  n56 = CollectionParameter() {collection=&FortniteMaterialParameters, parameterName=EventAudio_Amplitude}
  n57 = ScalarParameter() {defaultValue=20, group=Wrap_Muzak, parameterName=Muz_Event_Mask}
  n58 = Saturate(None=Max(A=Multiply(A=@n54, B=@n55) {constB=0.5}, B=Multiply(A=@n56, B=@n57) {constB=0.5}))
  n59 = Min(
   A=@n22
   B=Min(A=OneMinus(None=@n58), B=ScalarParameter() {defaultValue=1, group=Wrap_Muzak, parameterName=Muz_Note_For_Roll, sortPriority=600})
  )
  n60 = MaterialFunctionCall(UVs=@n48, Texture Scale=LinearInterpolate(A=@n51, Alpha=@n59) {constB=0}) {materialFunction=&ScaleUVsByCenter}
  n61 = Time()
  n62 = ScalarParameter() {defaultValue=1, group=Wrap_Muzak, parameterName=Muz_Note_Rotation_Speed}
  n63 = Multiply(
   A=Sine(None=Multiply(A=@n61, B=@n62))
   B=ScalarParameter() {defaultValue=0.01, group=Wrap_Muzak, parameterName=Muz_Note_Rotation_Sine_Amount}
  )
  n64 = TextureObject() {samplerType=SAMPLERTYPE_Masks, texture=&T_Wrap_Muzak_m}
  n65 = Multiply(
   A=TextureSample(
    UVs=MaterialFunctionCall(UVs=@n60.UVs, "Rotation Angle (0-1)"=@n63) {materialFunction=&CustomRotator}.Rotated Values
    Tex=@n64
   ) {samplerType=SAMPLERTYPE_Masks, texture=&T_Wrap_Muzak_m}.A
   B=@n60.0-1 mask
  )
  n66 = CurveAtlasRowParameter(CurveTime=@n65) {
   atlas=/Game/Weapons/WeaponSkins/Wraps/Textures/CruveAtlas_Wraps_01
   curve=/Game/Weapons/WeaponSkins/Wraps/Textures/Muzak/C_Muzak_Color1
   defaultValue=27
   group=Wrap_Muzak
   parameterName=Muz_Note_ColorCurve
  }
  n67 = MaterialFunctionCall!(Index=@n41, CCinput=@n66, CCinput_Alpha=@n66, UV=@n65) {materialFunction=&/MF_Global_CC_Tex}
  n68 = Multiply(A=@n51) {constB=1.2}
  n69 = Multiply(A=Add(A=Time()) {constA=0.03, constB=0.03}, B=ScalarParameter() {defaultValue=1, group=Wrap_Muzak, parameterName=Muz_Note_Scale_Speed})
  n70 = Multiply(A=Add(A=Sine(None=@n69))) {constB=0.5}
  n71 = MaterialFunctionCall(UVs=@n48, Texture Scale=LinearInterpolate(A=LinearInterpolate(A=@n68, B=@n51, Alpha=@n70) {constA=0.2}, Alpha=@n59) {constB=0}) {materialFunction=&ScaleUVsByCenter}
  n72 = Multiply(
   A=TextureSample(UVs=@n71.UVs, Tex=@n64) {samplerType=SAMPLERTYPE_Masks, texture=&T_Wrap_Muzak_m}.B
   B=@n71.0-1 mask
  )
  n73 = CurveAtlasRowParameter(CurveTime=@n72) {
   atlas=&CruveAtlas_Wraps_01
   curve=&C_Muzak_Color1
   defaultValue=27
   group=Wrap_Muzak
   parameterName=Muz_Note_ColorCurve
  }
  n74 = MaterialFunctionCall(Index=@n41, CCinput=@n73, CCinput_Alpha=@n73, UV=@n72) {materialFunction=&MF_Global_CC_Tex}
  n75 = OneMinus(None=Add(A=@n67.Alpha, B=@n74.Alpha))
  n76 = PreSkinnedNormal()
  n77 = Abs(None=FunctionInput(Preview=VertexInterpolator(VS=@n76).PS) {bUsePreviewValueAsDefault=true, inputName=Pre-Skinned Local Normal, sortPriority=10})
  n78 = ComponentMask(None=@n77) {r=true}
  n79 = ComponentMask(None=@n77) {g=true}
  n80 = LinearInterpolate(A=ComponentMask(None=@n77) {b=true}, B=@n79, Alpha=@n14)
  n81 = MaterialFunctionCall(
   In=LinearInterpolate(A=LinearInterpolate(A=@n78, B=@n80, Alpha=@n15), B=@n79, Alpha=@n16)
   Contrast=ScalarParameter() {defaultValue=1, group=Wrap_Muzak, parameterName=Muz_Tri_Planner_Mask_Contrast, sortPriority=0}
  ) {materialFunction=&CheapContrast}
  n82 = MaterialFunctionCall(Float3=NamedRerouteUsage()) {materialFunction=&BreakOutFloat3Components}
  n83 = Max(A=@n82.R, B=@n82.R)
  n84 = ShadingPathSwitch(Default=@n39.A, Deferred=@n39.A, Mobile=LinearInterpolate(A=NamedRerouteUsage(), B=@n39.A, Alpha=Max(A=@n83, B=@n82.B))) {desc=Migrated from FeatureLevelSwitch}
  n85 = CurveAtlasRowParameter(CurveTime=@n84) {
   atlas=&CruveAtlas_Wraps_01
   curve=&C_Muzak_Color1
   defaultValue=27
   group=Wrap_Muzak
   parameterName=Muz_Line_ColorCurve
  }
  n86 = OneMinus(None=@n85.A)
  n87 = ScalarParameter() {defaultValue=40, group=Wrap_Muzak, parameterName=Muz_CellsX}
  n88 = MaterialFunctionCall(X=@n87, Y=ScalarParameter() {defaultValue=1, group=Wrap_Muzak, parameterName=Muz_CellsY}) {materialFunction=&MakeFloat2}
  n89 = MaterialFunctionCall(Color=@n8) {materialFunction=&SplitComponents}
  n90 = AppendVector(A=@n89.R, B=@n89.R)
  n91 = OneMinus(None=@n89.B)
  n92 = LinearInterpolate(
   A=AppendVector(A=@n89.G, B=OneMinus(None=@n89.B))
   B=LinearInterpolate(
    A=MaterialFunctionCall(UVs=@n90, "Rotation Angle (0-1)"=@n10) {materialFunction=&CustomRotator}.Rotated Values
    B=AppendVector(A=@n89.R, B=@n91)
    Alpha=@n14
   )
   Alpha=@n15
  )
  n93 = AppendVector(A=@n89.B, B=@n89.B)
  n94 = Multiply(A=Frac(None=LinearInterpolate(A=@n92, B=@n93, Alpha=@n16)), B=@n88.Result) {constB=5}
  n95 = Floor(None=ComponentMask(None=@n94) {r=true})
  n96 = ComponentMask(None=@n94) {g=true}
  n97 = MaterialFunctionCall(Float2=Divide(A=AppendVector(A=@n95, B=@n96), B=@n88.Result) {constB=5}) {materialFunction=&BreakOutFloat2Components}
  n98 = OneMinus(None=Multiply(A=@n97.G) {constB=2})
  n99 = AppendVector(
   A=Multiply(A=@n97.R, B=ScalarParameter() {defaultValue=3.5, group=Wrap_Muzak, parameterName=Muz_U_mul, sortPriority=3}) {constB=0.5}
   B=OneMinus(None=Abs(None=@n98))
  )
  n100 = LinearInterpolate(
   A=ScalarParameter() {defaultValue=1, group=Wrap_Muzak, parameterName=Muz_Equalizer_Mask_G_Weapon}
   B=LinearInterpolate(
    A=ScalarParameter() {defaultValue=1, group=Wrap_Muzak, parameterName=Muz_Equalizer_Mask_G_Vehicle}
    B=ScalarParameter() {defaultValue=1, group=Wrap_Muzak, parameterName=Muz_Equalizer_Mask_G_VehicleSD}
    Alpha=@n14
   )
   Alpha=@n15
  )
  n101 = ScalarParameter() {defaultValue=1, group=Wrap_Muzak, parameterName=Muz_Equalizer_Mask_G_Mech}
  MaterialExpressionAppendVector_7 = AppendVector(
   A=Frac(None=ComponentMask(None=@n99) {r=true})
   B=Multiply(A=LinearInterpolate(A=@n100, B=@n101, Alpha=@n16), B=ComponentMask(None=@n99) {g=true})
  )
  n102 = LinearInterpolate(A=@n8, B=Constant3Vector() {constant=<rgba 0 1 0 0>}, Alpha=@n31)
  n103 = Abs(None=OneMinus(None=Multiply(A=@n102) {constB=2}))
  n104 = FunctionInput(Preview=Constant3Vector() {constant=<rgba 0 1 0 0>}) {bUsePreviewValueAsDefault=true, inputName=ForwardAxis}
  n105 = MaterialFunctionCall!(Bounds=@n103, ForwardAxis=@n104, Offset=Constant() {r=-0.25}, FallOff=Constant()) {materialFunction=&/MF_Wrap_Gradent_Simple}
  n106 = OneMinus(None=@n105.Clamped_Data)
  n107 = MaterialFunctionCall(Bounds=@n103, ForwardAxis=@n104, Offset=Constant() {r=-0.75}, FallOff=Constant()) {materialFunction=&MF_Wrap_Gradent_Simple}
  n108 = MaterialFunctionCall(Color=@n8) {materialFunction=&SplitComponents}
  n109 = LinearInterpolate(
   A=AppendVector(A=@n108.G, B=@n108.G)
   B=LinearInterpolate(A=AppendVector(A=@n108.R, B=@n108.R), B=AppendVector(A=@n108.R, B=@n108.R), Alpha=@n14)
   Alpha=@n15
  )
  n110 = AppendVector(A=@n108.B, B=@n108.B)
  n111 = Multiply(A=ComponentMask(None=LinearInterpolate(A=@n109, B=@n110, Alpha=@n16)) {g=true}) {constB=2}
  n112 = Saturate(None=Divide(A=OneMinus(None=@n111)) {constB=0})
  n113 = CollectionParameter() {collection=&FortniteMaterialParameters, parameterName=EventAudio_Spectrum}
  n114 = ScalarParameter() {group=Wrap_Muzak, parameterName=Muz_Sound_Specturm strength, sortPriority=600}
  n115 = MaterialFunctionCall(Float4=Power(
    Base=Multiply(A=@n113, B=@n114) {constB=0.65}
    Exp=ScalarParameter() {defaultValue=1, group=Wrap_Muzak, parameterName=Muz_Sound_Specturm Power, sortPriority=600}
   )) {materialFunction=&BreakOutFloat4Components}
  n116 = Multiply(A=LinearInterpolate(A=@n106, B=@n107.Clamped_Data, Alpha=@n112), B=@n115.R)
  n117 = MaterialFunctionCall(Bounds=@n103, ForwardAxis=@n104, Offset=Constant() {r=-0.5}, FallOff=Constant()) {materialFunction=&MF_Wrap_Gradent_Simple}
  n118 = Multiply(A=OneMinus(None=@n117.Clamped_Data), B=@n105.Clamped_Data)
  n119 = Multiply(A=OneMinus(None=@n107.Clamped_Data), B=@n117.Clamped_Data)
  n120 = Multiply(A=LinearInterpolate(A=@n118, B=@n119, Alpha=@n112), B=@n115.G)
  n121 = Multiply(A=LinearInterpolate(A=@n119, B=@n118, Alpha=@n112), B=@n115.B)
  n122 = Multiply(A=LinearInterpolate(A=@n107.Clamped_Data, B=@n106, Alpha=@n112), B=@n115.A)
  n123 = ScalarParameter() {defaultValue=1, group=Wrap_Muzak, parameterName=Muz_Sound_ampltude_effect, sortPriority=600}
  n124 = Add(
   A=Power(
    Base=Multiply(A=@n56, B=@n123) {constB=0.5}
    Exp=ScalarParameter() {defaultValue=1, group=Wrap_Muzak, parameterName=Muz_Sound_ampltude_Power, sortPriority=600}
   )
   B=Add(A=Add(A=@n116, B=@n120), B=Add(A=@n121, B=@n122))
  )
  MaterialExpressionMax_2 = Max(A=@n124) {constB=0}
  MaterialExpressionReroute_26 = Reroute(None=Max(A=@n124) {constB=0})
  n125 = ComponentMask(None=@n99) {r=true}
  MaterialExpressionLinearInterpolate_0 = LinearInterpolate(A=@MaterialExpressionMax_2, B=@MaterialExpressionReroute_26, Alpha=@n125)
  MaterialExpressionLinearInterpolate_7 = LinearInterpolate(A=@MaterialExpressionMax_2, B=@MaterialExpressionReroute_26, Alpha=@n125)
  MaterialExpressionLinearInterpolate_8 = LinearInterpolate(A=@MaterialExpressionMax_2, B=@MaterialExpressionReroute_26, Alpha=@n125)
  MaterialExpressionScalarParameter_16 = ScalarParameter() {defaultValue=3.5, group=Wrap_Muzak, parameterName=Muz_BandsWidth}
  MaterialExpressionCustom_0 = Custom!(
   uv=@MaterialExpressionAppendVector_7
   waveIntensity0=@MaterialExpressionMax_2
   waveIntensity1=@MaterialExpressionLinearInterpolate_0
   waveIntensity2=@MaterialExpressionLinearInterpolate_7
   waveIntensity3=@MaterialExpressionLinearInterpolate_8
   waveIntensity4=@MaterialExpressionReroute_26
   waveWidth=@MaterialExpressionScalarParameter_16
  ) {
   code="float waveA = -pow((waveWidth * uv.x - 0), 2) + waveIntensity0;\r\nfloat waveB = -pow((waveWidth * uv.x - waveWidth * 0.25), 2) + waveIntensity1;\r\nfloat waveC = -pow((waveWidth * uv.x - waveWidth * 0.5), 2) + waveIntensity2;\r\nfloat waveD = -pow((waveWidth * uv.x - waveWidth * 0.75), 2) + waveIntensity3;\r\nfloat waveE = -pow((waveWidth * uv.x - waveWidth), 2) + waveIntensity4;\r\nreturn max(waveA,max(waveB,max(waveC,max(waveD,waveE))));"
   description=CreateWave
   outputType=CMOT_Float1
   inputs [
    {inputName=uv, input {expression=&self:MaterialExpressionAppendVector_7}}
    {inputName=waveIntensity0, input {expression=&self:MaterialExpressionMax_2}}
    {
     inputName=waveIntensity1
     input {expression=&self:MaterialExpressionLinearInterpolate_0}
    }
    {
     inputName=waveIntensity2
     input {expression=&self:MaterialExpressionLinearInterpolate_7}
    }
    {
     inputName=waveIntensity3
     input {expression=&self:MaterialExpressionLinearInterpolate_8}
    }
    {inputName=waveIntensity4, input {expression=&self:MaterialExpressionReroute_26}}
    {inputName=waveWidth, input {expression=&self:MaterialExpressionScalarParameter_16}}
   ]
  }
  n126 = ScalarParameter() {group=Wrap_Muzak, parameterName=Muz_Equalizer_VerticalOffset_Vehicle}
  n127 = ScalarParameter() {group=Wrap_Muzak, parameterName=Muz_Equalizer_VerticalOffset_VehicleSD}
  MaterialExpressionLinearInterpolate_1 = LinearInterpolate(
   A=LinearInterpolate(
    A=ScalarParameter() {group=Wrap_Muzak, parameterName=Muz_Equalizer_VerticalOffset_Weapon}
    B=LinearInterpolate(A=@n126, B=@n127, Alpha=@n14)
    Alpha=@n15
   )
   B=ScalarParameter() {group=Wrap_Muzak, parameterName=Muz_Equalizer_VerticalOffset_Mech}
   Alpha=@n16
  )
  n128 = Custom!(uv=@MaterialExpressionAppendVector_7, waveFunction=@MaterialExpressionCustom_0, verticalOffset=@MaterialExpressionLinearInterpolate_1) {
   code="float result = uv.y- verticalOffset + waveFunction;    \r\nreturn (0.1, result);"
   description=DrawWave
   outputType=CMOT_Float1
   inputs [
    {inputName=uv, input {expression=&self:MaterialExpressionAppendVector_7}}
    {inputName=waveFunction, input {expression=&self:MaterialExpressionCustom_0}}
    {
     inputName=verticalOffset
     input {expression=&self:MaterialExpressionLinearInterpolate_1}
    }
   ]
  }
  n129 = Time()
  n130 = ScalarParameter() {group=Wrap_Muzak, parameterName=Muz_Sine_X_Speed}
  n131 = MaterialFunctionCall(Color=Multiply(A=@n8, B=Constant3Vector() {constant=<rgba 1 1 -1 0>})) {materialFunction=&SplitComponents}
  n132 = OneMinus(None=@n131.B)
  n133 = MaterialFunctionCall(UVs=AppendVector(A=@n131.R, B=@n131.R), "Rotation Angle (0-1)"=@n10) {materialFunction=&CustomRotator}
  n134 = AppendVector(A=@n131.R, B=OneMinus(None=@n131.B))
  n135 = LinearInterpolate(
   A=LinearInterpolate(A=AppendVector(A=@n131.G, B=@n132), B=LinearInterpolate(A=@n133.Rotated Values, B=@n134, Alpha=@n14), Alpha=@n15)
   B=AppendVector(A=@n131.B, B=@n131.B)
   Alpha=@n16
  )
  n136 = ScalarParameter() {defaultValue=2, group=Wrap_Muzak, parameterName=Muz_Sine_X_tile}
  n137 = ComponentMask(None=Add(A=Multiply(A=@n129, B=@n130), B=Multiply(A=@n135, B=@n136))) {r=true}
  n138 = ScalarParameter() {defaultValue=1, group=Wrap_Muzak, parameterName=Muz_Sine_Y_Tile_Weapon}
  n139 = LinearInterpolate(B=ScalarParameter() {defaultValue=1, group=Wrap_Muzak, parameterName=Muz_Sine_Y_Tile_Vehicle_Side}, Alpha=@n14) {constA=1}
  n140 = ComponentMask(None=MaterialFunctionCall(UVs=@n135, Texture Scale=LinearInterpolate(A=@n138, B=@n139, Alpha=@n15)) {materialFunction=&ScaleUVsByCenter}.UVs) {g=true}
  n141 = LinearInterpolate(A=TextureSample(UVs=AppendVector(A=@n137, B=@n140), Tex=@n18).B, B=@n128, Alpha=@n58)
  n142 = OneMinus(None=@n141)
  n143 = ScalarParameter() {defaultValue=220, group=Wrap_Muzak, parameterName=Muz_Equ_Dot_Contrast}
  n144 = ComponentMask(None=@n17) {r=true}
  n145 = TextureSample(UVs=AppendVector(A=Multiply(A=@n144, B=@n87), B=ComponentMask(None=@n17) {g=true}), Tex=@n64) {samplerType=SAMPLERTYPE_Masks, texture=&T_Wrap_Muzak_m}
  n146 = Multiply(
   A=@n81.Result
   B=MaterialFunctionCall(In=@n141, Contrast=ScalarParameter() {defaultValue=10, group=Wrap_Muzak, parameterName=Muz_Equ_Contrast}) {materialFunction=&CheapContrast}.Result
  )
  n147 = Multiply(
   A=Multiply(
    A=MaterialFunctionCall(In=@n142, Contrast=@n143) {materialFunction=&CheapContrast}.Result
    B=@n145.R
   )
   B=@n146
  )
  n148 = Multiply(A=@n145.R, B=@n146)
  n149 = VectorParameter() {group=Wrap_Muzak, parameterName=Muz_BackGround_Color_Bck, sortPriority=0, defaultValue=<rgba 0.0065 0.0065 0.0666 1>}
  n150 = VectorParameter() {group=Wrap_Muzak, parameterName=Muz_BackGround_Color_Frnt, sortPriority=0, defaultValue=<rgba 0.0536 0.1702 0.3854 1>}
  n151 = Constant() {r=-0.25}
  n152 = Constant() {r=0.75}
  n153 = LinearInterpolate(
   A=@n149.RGB
   B=@n150.RGB
   Alpha=Power(Base=MaterialFunctionCall(Bounds=@n102, ForwardAxis=@n104, Offset=@n151, FallOff=@n152) {materialFunction=&MF_Wrap_Gradent_Simple}.Clamped_Data) {constExponent=3}
  )
  n154 = Time()
  n155 = ComponentMask(None=VectorParameter() {group=Wrap_Muzak, parameterName=Muz_Equ_Gradent_Speed, defaultValue=<rgba 0 0 0 1>}.RGB) {g=true, r=true}
  n156 = TextureSample(UVs=Add(A=@n17, B=Multiply(A=@n154, B=@n155)), Tex=@n18) {samplerType=SAMPLERTYPE_Masks, texture=&T_Wrap_Muzak_m}
  n157 = CurveAtlasRowParameter(CurveTime=@n156.R) {
   atlas=&CruveAtlas_Wraps_01
   curve=/Game/Weapons/WeaponSkins/Wraps/Textures/Muzak/C_Muzak_Color
   defaultValue=26
   group=Wrap_Muzak
   parameterName=Muz_Equ_ColorCurve
  }
  n158 = MaterialFunctionCall(Index=ScalarParameter() {defaultValue=26, group=Wrap_Muzak, parameterName="Muz_Equ_ColorCurve  Index"}, CCinput=@n157, UV=@n156.R) {materialFunction=&MF_Global_CC_Tex}
  n159 = TextureSample(UVs=@n17, Tex=@n38) {samplerType=SAMPLERTYPE_Masks, texture=&T_Wrap_Muzak_m}
  n160 = CurveAtlasRowParameter(CurveTime=@n159.R) {
   atlas=&CruveAtlas_Wraps_01
   curve=/Game/Weapons/WeaponSkins/Wraps/Textures/Muzak/C_Muzak_Color2
   defaultValue=28
   group=Wrap_Muzak
   parameterName=Muz_Line_ColorCurve_Bkgnd
  }
  n161 = MaterialFunctionCall(
   Index=ScalarParameter() {defaultValue=28, group=Wrap_Muzak, parameterName="Muz_Line_ColorCurve_Bkgnd  Index"}
   CCinput=@n160
   CCinput_Alpha=@n160
   UV=@n159.R
  ) {materialFunction=&MF_Global_CC_Tex}
  n162 = Add(A=@n147, B=LinearInterpolate(A=LinearInterpolate(A=@n153, B=@n158.3V, Alpha=@n148), B=@n161.3V, Alpha=@n161.3V))
  n163 = ScalarParameter() {defaultValue=1, group=Wrap_Muzak, parameterName=Muz_Emissive_Brightness, sortPriority=-1}
  n164 = MaterialFunctionCall(
   Index=ScalarParameter() {defaultValue=27, group=Wrap_Muzak, parameterName="Muz_Line_ColorCurve  Index"}
   CCinput=@n85
   CCinput_Alpha=@n85
   UV=@n84
  ) {materialFunction=&MF_Global_CC_Tex}
  n165 = Add(A=Add(A=Multiply(A=@n67.3V, B=@n67.3V), B=Multiply(A=@n74.3V, B=@n74.3V)), B=Add(A=Multiply(A=@n164.3V, B=@n164.3V), B=@n147))
  n166 = FunctionInput() {inputName=Material, inputType=FunctionInput_MaterialAttributes}
  n167 = GetMaterialAttributes(None=@n166) {attributeGetTypes [#1, #2, #3]}
  n168 = SetMaterialAttributes!(
   MaterialAttributes=@n167.MaterialAttributes
   BaseColor=Multiply(A=LinearInterpolate(B=@n75, Alpha=@n81.Result) {constA=1}, B=Multiply(A=@n86, B=@n162))
   EmissiveColor=Multiply(A=@n81.Result, B=Multiply(A=@n163, B=@n165))
   Roughness=ScalarParameter() {defaultValue=0.3, group=Wrap_Muzak, parameterName=Muz_Roughness, sortPriority=-1}
   Specular=ScalarParameter() {defaultValue=0.5, group=Wrap_Muzak, parameterName=Muz_Specular, sortPriority=-1}
   Metallic=Constant()
  ) {attributeSetTypes [
    #1
    #4
    #2
    #5
    #6
   ]}
  n169 = Saturate(None=Add(
    A=FunctionInput() {bUsePreviewValueAsDefault=true, inputName="Customization Mask (_CM B)", inputType=FunctionInput_Scalar, sortPriority=24}
    B=ScalarParameter() {group=Wrap_Muzak, parameterName=Muz_Remove_Customization Mask, sortPriority=-10}
   ))
  FunctionOutput(None=StaticSwitchParameter(True=BlendMaterialAttributes(A=@n167.MaterialAttributes, B=@n168, Alpha=@n169), False=@n166) {group=Customization_Wrap, parameterName=Wrap_Muzak})
  n170 = NamedRerouteDeclaration(None=@n15) {name=isVehicle, nodeColor=<rgba 0 0.8125 1 1>}
  n171 = NamedRerouteDeclaration(None=@n14) {name=isVehicleSide, nodeColor=<rgba 0.125 1 0 1>}
  NamedRerouteDeclaration(None=AppendVector(A=AppendVector(A=@n170, B=@n171), B=NamedRerouteDeclaration(None=@n16) {name=isVehicleMech, nodeColor=<rgba 0 1 0.3203 1>})) {name=isState, nodeColor=<rgba 0 0.2031 1 1>}
  NamedRerouteDeclaration(None=@n17) {name=Uvs, nodeColor=<rgba 1 0.1172 0 1>}
  n172 = ComponentMask(None=@n40) {r=true}
  n173 = ScalarParameter() {defaultValue=-0.99, group=Wrap_Muzak_Switch_Line, parameterName=Muz_Switch_Fade_Start}
  n174 = Saturate(None=Divide(
    A=Add(A=@n172, B=@n173) {constB=-0.99}
    B=ScalarParameter() {defaultValue=-0.1, group=Wrap_Muzak_Switch_Line, parameterName=Muz_Switch_Falloff_Start}
   ) {constB=-0.1})
  n175 = Add(A=@n172, B=ScalarParameter() {defaultValue=-0.01, group=Wrap_Muzak_Switch_Line, parameterName=Muz_Switch_Fade_End}) {constB=-0.01}
  n176 = ScalarParameter() {defaultValue=0.1, group=Wrap_Muzak_Switch_Line, parameterName=Muz_Switch_Falloff_End}
  n177 = Frac(None=Add(A=ComponentMask(None=@n40) {g=true}, B=NamedRerouteUsage()))
  n178 = ScalarParameter() {defaultValue=-0.49, group=Wrap_Muzak_Switch_Line, parameterName=Muz_Switch_Line_Thickness}
  n179 = ScalarParameter() {defaultValue=0.005, group=Wrap_Muzak_Switch_Line, parameterName=Muz_Switch_Line_Falloff}
  n180 = Divide(A=Add(A=@n177, B=@n178) {constB=0.2}, B=@n179) {constB=0.05}
  n181 = Divide(A=Add(A=OneMinus(None=@n177), B=@n178), B=@n179) {constB=0.05}
  n182 = Multiply(A=Saturate(None=Divide(A=@n175, B=@n176) {constB=0.1}), B=Multiply(A=Saturate(None=@n180), B=Saturate(None=@n181)))
  NamedRerouteDeclaration(None=Saturate(None=Multiply(A=@n174, B=@n182))) {name=Proc_line, nodeColor=<rgba 0 0.0859 1 1>}
 ]}
