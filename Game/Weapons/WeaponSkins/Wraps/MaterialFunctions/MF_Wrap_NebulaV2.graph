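# Material function for the NebulaV2 weapon wrap, layering a nebula effect onto an incoming material through attribute blends.
# Primary, secondary and crevice masks, channel-selected warp and blend masks, and Fresnel shape the tinted, optionally desaturated texture layers.
# Chained lerps pick color and scalar parameters per state, apparently from the IsState inputs.
# Bounds, actor position and a safe-zone virtual pivot inform positioning, with camera FOV remapped as a factor.
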MaterialFunction /Game/Weapons/WeaponSkins/Wraps/MaterialFunctions/MF_Wrap_NebulaV2
graph {expressions {
  MaterialExpressionAbs_1 {class=MaterialExpressionAbs, inputs [None {from=MaterialExpressionNamedRerouteUsage_83, output=""}], properties {}}
  MaterialExpressionActorPositionWS_3 {class=MaterialExpressionActorPositionWS, inputs [], properties {}}
  MaterialExpressionAdd_0 {
   class=MaterialExpressionAdd
   inputs [
    A {from=MaterialExpressionNamedRerouteUsage_13}
    B {from=""}
   ]
   properties {constB=-0.5}
  }
  MaterialExpressionAdd_1 {
   class=MaterialExpressionAdd
   inputs [
    A {from=MaterialExpressionDivide_0}
    B {from=MaterialExpressionMultiply_1}
   ]
  }
  MaterialExpressionAdd_10 {
   class=MaterialExpressionAdd
   inputs [
    A {from=MaterialExpressionComponentMask_0}
    B {from=MaterialExpressionNamedRerouteUsage_62}
   ]
  }
  MaterialExpressionAdd_11 {
   class=MaterialExpressionAdd
   inputs [
    A {from=MaterialExpressionDivide_0}
    B {from=MaterialExpressionMultiply_11}
   ]
  }
  MaterialExpressionAdd_12 {
   class=MaterialExpressionAdd
   inputs [
    A {from=MaterialExpressionFunctionInput_18}
    B {from=MaterialExpressionFunctionInput_11}
   ]
  }
  MaterialExpressionAdd_13 {
   class=MaterialExpressionAdd
   inputs [
    A {from=MaterialExpressionMaterialFunctionCall_50, output=Result}
    B {from=MaterialExpressionScalarParameter_65}
   ]
  }
  MaterialExpressionAdd_14 {
   class=MaterialExpressionAdd
   inputs [
    A {from=MaterialExpressionNamedRerouteUsage_68}
    B {from=MaterialExpressionMultiply_46}
   ]
  }
  MaterialExpressionAdd_15 {
   class=MaterialExpressionAdd
   inputs [
    A {from=MaterialExpressionMultiply_14}
    B {from=MaterialExpressionNamedRerouteUsage_86}
   ]
  }
  MaterialExpressionAdd_16 {
   class=MaterialExpressionAdd
   inputs [
    A {from=MaterialExpressionNamedRerouteUsage_87}
    B {from=""}
   ]
   properties {constB=-0.5}
  }
  MaterialExpressionAdd_17 {
   class=MaterialExpressionAdd
   inputs [
    A {from=MaterialExpressionNamedRerouteUsage_88}
    B {from=""}
   ]
   properties {constB=-0.5}
  }
  MaterialExpressionAdd_18 {
   class=MaterialExpressionAdd
   inputs [
    A {from=MaterialExpressionMultiply_0}
    B {from=MaterialExpressionNamedRerouteUsage_101}
   ]
  }
  MaterialExpressionAdd_19 {
   class=MaterialExpressionAdd
   inputs [
    A {from=MaterialExpressionActorPositionWS_3}
    B {from=MaterialExpressionConstant3Vector_5}
   ]
  }
  MaterialExpressionAdd_2 {
   class=MaterialExpressionAdd
   inputs [
    A {from=MaterialExpressionOneMinus_0}
    B {from=""}
   ]
   properties {constB=-0.8}
  }
  MaterialExpressionAdd_20 {
   class=MaterialExpressionAdd
   inputs [
    A {from=MaterialExpressionChannelMaskParameter_1, output=R}
    B {from=MaterialExpressionScalarParameter_69}
   ]
  }
  MaterialExpressionAdd_21 {
   class=MaterialExpressionAdd
   inputs [
    A {from=MaterialExpressionMultiply_26}
    B {from=MaterialExpressionMaterialFunctionCall_32, output=Result}
   ]
  }
  MaterialExpressionAdd_22 {
   class=MaterialExpressionAdd
   inputs [
    A {from=MaterialExpressionMultiply_57}
    B {from=MaterialExpressionMultiply_56}
   ]
  }
  MaterialExpressionAdd_23 {
   class=MaterialExpressionAdd
   inputs [
    A {from=MaterialExpressionAdd_22}
    B {from=MaterialExpressionScalarParameter_72}
   ]
  }
  MaterialExpressionAdd_24 {
   class=MaterialExpressionAdd
   inputs [
    A {from=MaterialExpressionAdd_1}
    B {from=MaterialExpressionNamedRerouteUsage_105}
   ]
  }
  MaterialExpressionAdd_25 {
   class=MaterialExpressionAdd
   inputs [
    A {from=MaterialExpressionLinearInterpolate_80}
    B {from=MaterialExpressionStaticSwitchParameter_1}
   ]
  }
  MaterialExpressionAdd_3 {
   class=MaterialExpressionAdd
   inputs [
    A {from=MaterialExpressionMultiply_24}
    B {from=MaterialExpressionAdd_19}
   ]
  }
  MaterialExpressionAdd_4 {
   class=MaterialExpressionAdd
   inputs [
    A {from=MaterialExpressionNamedRerouteUsage_16}
    B {from=""}
   ]
   properties {constB=-0.5}
  }
  MaterialExpressionAdd_5 {
   class=MaterialExpressionAdd
   inputs [
    A {from=MaterialExpressionNamedRerouteUsage_39}
    B {from=MaterialExpressionNamedRerouteUsage_41}
   ]
  }
  MaterialExpressionAdd_6 {
   class=MaterialExpressionAdd
   inputs [
    A {from=MaterialExpressionMultiply_39}
    B {from=""}
   ]
   properties {constB=0.5}
  }
  MaterialExpressionAdd_7 {
   class=MaterialExpressionAdd
   inputs [
    A {from=MaterialExpressionNamedRerouteUsage_46}
    B {from=""}
   ]
   properties {constB=-0.5}
  }
  MaterialExpressionAdd_8 {
   class=MaterialExpressionAdd
   inputs [
    A {from=MaterialExpressionMaterialFunctionCall_49, output=Result}
    B {from=MaterialExpressionScalarParameter_79}
   ]
  }
  MaterialExpressionAdd_9 {
   class=MaterialExpressionAdd
   inputs [
    A {from=MaterialExpressionVectorParameter_35, output=RGB}
    B {from=MaterialExpressionMultiply_44}
   ]
  }
  MaterialExpressionAppendVector_0 {
   class=MaterialExpressionAppendVector
   inputs [
    A {from=MaterialExpressionVectorParameter_5, output=RGB}
    B {from=MaterialExpressionVectorParameter_5, output=RGB}
   ]
  }
  MaterialExpressionAppendVector_1 {
   class=MaterialExpressionAppendVector
   inputs [
    A {from=MaterialExpressionVectorParameter_19, output=RGB}
    B {from=MaterialExpressionVectorParameter_19, output=RGB}
   ]
  }
  MaterialExpressionAppendVector_10 {
   class=MaterialExpressionAppendVector
   inputs [
    A {from=MaterialExpressionVectorParameter_17, output=RGB}
    B {from=MaterialExpressionVectorParameter_17, output=RGB}
   ]
  }
  MaterialExpressionAppendVector_11 {
   class=MaterialExpressionAppendVector
   inputs [
    A {from=MaterialExpressionScalarParameter_51, output=""}
    B {from=MaterialExpressionScalarParameter_52, output=""}
   ]
  }
  MaterialExpressionAppendVector_12 {
   class=MaterialExpressionAppendVector
   inputs [
    A {from=MaterialExpressionAppendVector_11, output=""}
    B {from=MaterialExpressionScalarParameter_53, output=""}
   ]
  }
  MaterialExpressionAppendVector_13 {
   class=MaterialExpressionAppendVector
   inputs [
    A {from=MaterialExpressionVectorParameter_22, output=RGB}
    B {from=MaterialExpressionVectorParameter_22, output=RGB}
   ]
  }
  MaterialExpressionAppendVector_2 {
   class=MaterialExpressionAppendVector
   inputs [
    A {from=MaterialExpressionVectorParameter_21, output=RGB}
    B {from=MaterialExpressionVectorParameter_21, output=RGB}
   ]
  }
  MaterialExpressionAppendVector_3 {
   class=MaterialExpressionAppendVector
   inputs [
    A {from=MaterialExpressionConstant_1, output=""}
    B {from=MaterialExpressionDivide_33, output=""}
   ]
  }
  MaterialExpressionAppendVector_7 {
   class=MaterialExpressionAppendVector
   inputs [
    A {from=MaterialExpressionVectorParameter_24, output=RGB}
    B {from=MaterialExpressionVectorParameter_24, output=RGB}
   ]
  }
  MaterialExpressionAppendVector_8 {
   class=MaterialExpressionAppendVector
   inputs [
    A {from=MaterialExpressionVectorParameter_25, output=RGB}
    B {from=MaterialExpressionVectorParameter_25, output=RGB}
   ]
  }
  MaterialExpressionAppendVector_9 {
   class=MaterialExpressionAppendVector
   inputs [
    A {from=MaterialExpressionVectorParameter_20, output=RGB}
    B {from=MaterialExpressionVectorParameter_20, output=RGB}
   ]
  }
  MaterialExpressionBlendMaterialAttributes_1 {
   class=MaterialExpressionBlendMaterialAttributes
   inputs [
    A {from=MaterialExpressionNamedRerouteUsage_104, output=""}
    B {from=MaterialExpressionNamedRerouteUsage_3, output=""}
    Alpha {from=MaterialExpressionNamedRerouteDeclaration_3, output=""}
   ]
   properties {vertexAttributeBlendType=UseA}
  }
  MaterialExpressionBlendMaterialAttributes_2 {
   class=MaterialExpressionBlendMaterialAttributes
   inputs [
    A {from=MaterialExpressionStaticSwitchParameter_0, output=""}
    B {from=MaterialExpressionNamedRerouteUsage_71, output=""}
    Alpha {from=MaterialExpressionMultiply_58, output=""}
   ]
   properties {vertexAttributeBlendType=UseA}
  }
  MaterialExpressionChannelMaskParameter_0 {
   class=MaterialExpressionChannelMaskParameter
   inputs [None {from=MaterialExpressionMaterialFunctionCall_34, output=Combined}]
   properties {group=Wrap_NebulaV2_Blend, maskChannel=Alpha, parameterName=Neb_Blend_Warp_Channel, defaultValue=<rgba 0 0 0 1>}
  }
  MaterialExpressionChannelMaskParameter_1 {
   class=MaterialExpressionChannelMaskParameter
   inputs [None {from=MaterialExpressionMaterialFunctionCall_34, output=Combined}]
   properties {group=Wrap_NebulaV2_Blend, maskChannel=Alpha, parameterName=Neb_Blend_Mask_Channel, defaultValue=<rgba 0 0 0 1>}
  }
  MaterialExpressionClamp_0 {
   class=MaterialExpressionClamp
   inputs [
    None {from=MaterialExpressionAdd_4, output=""}
    Min {from=MaterialExpressionScalarParameter_80, output=""}
    Max {from="", output=""}
   ]
   properties {minDefault=-0.1}
  }
  MaterialExpressionComponentMask_0 {class=MaterialExpressionComponentMask, inputs [None {from=MaterialExpressionMaterialFunctionCall_15, output=UVs}], properties {r=true}}
  MaterialExpressionComponentMask_1 {class=MaterialExpressionComponentMask, inputs [None {from=MaterialExpressionNamedRerouteUsage_79}], properties {b=true, g=true, r=true}}
  MaterialExpressionComponentMask_10 {class=MaterialExpressionComponentMask, inputs [None {from=MaterialExpressionVectorParameter_44, output=RGB}], properties {g=true, r=true}}
  MaterialExpressionComponentMask_11 {class=MaterialExpressionComponentMask, inputs [None {from=MaterialExpressionVectorParameter_45, output=RGB}], properties {g=true, r=true}}
  MaterialExpressionComponentMask_2 {class=MaterialExpressionComponentMask, inputs [None {from=MaterialExpressionNamedRerouteUsage_79}], properties {a=true}}
  MaterialExpressionComponentMask_3 {class=MaterialExpressionComponentMask, inputs [None {from=MaterialExpressionVectorParameter_3, output=RGB}], properties {g=true, r=true}}
  MaterialExpressionComponentMask_4 {class=MaterialExpressionComponentMask, inputs [None {from=MaterialExpressionVectorParameter_31, output=RGB}], properties {g=true, r=true}}
  MaterialExpressionComponentMask_5 {class=MaterialExpressionComponentMask, inputs [None {from=MaterialExpressionMaterialFunctionCall_13, output=Result}], properties {b=true}}
  MaterialExpressionComponentMask_6 {class=MaterialExpressionComponentMask, inputs [None {from=MaterialExpressionMaterialFunctionCall_12, output=Result}], properties {r=true}}
  MaterialExpressionComponentMask_7 {class=MaterialExpressionComponentMask, inputs [None {from=MaterialExpressionNamedRerouteUsage_43}], properties {b=true, g=true, r=true}}
  MaterialExpressionComponentMask_8 {class=MaterialExpressionComponentMask, inputs [None {from=MaterialExpressionReroute_3}], properties {a=true}}
  MaterialExpressionComponentMask_9 {class=MaterialExpressionComponentMask, inputs [None {from=MaterialExpressionReroute_1}], properties {b=true, g=true, r=true}}
  MaterialExpressionConstant2Vector_3 {class=MaterialExpressionConstant2Vector}
  MaterialExpressionConstant3Vector_1 {class=MaterialExpressionConstant3Vector, properties {constant=<rgba 1 0 0 0>}}
  MaterialExpressionConstant3Vector_4 {class=MaterialExpressionConstant3Vector, properties {constant=<rgba 1 1 0 0>}}
  MaterialExpressionConstant3Vector_5 {class=MaterialExpressionConstant3Vector, properties {constant=<rgba 0 0 90 1>}}
  MaterialExpressionConstant4Vector_0 {class=MaterialExpressionConstant4Vector, inputs [], properties {}}
  MaterialExpressionConstant4Vector_4 {class=MaterialExpressionConstant4Vector, inputs [], properties {constant=<rgba 0 0 5 -1>}}
  MaterialExpressionConstant_0 {class=MaterialExpressionConstant, properties {r=0.5}}
  MaterialExpressionConstant_1 {class=MaterialExpressionConstant, properties {r=1}}
  MaterialExpressionConstant_10 {class=MaterialExpressionConstant}
  MaterialExpressionConstant_11 {class=MaterialExpressionConstant}
  MaterialExpressionConstant_12 {class=MaterialExpressionConstant, properties {r=0.25}}
  MaterialExpressionConstant_13 {class=MaterialExpressionConstant}
  MaterialExpressionConstant_14 {class=MaterialExpressionConstant, properties {r=-2500}}
  MaterialExpressionConstant_2 {class=MaterialExpressionConstant}
  MaterialExpressionConstant_3 {class=MaterialExpressionConstant, properties {r=45}}
  MaterialExpressionConstant_32 {class=MaterialExpressionConstant, properties {r=45}}
  MaterialExpressionConstant_33 {class=MaterialExpressionConstant, properties {r=90}}
  MaterialExpressionConstant_4 {class=MaterialExpressionConstant, properties {r=0.5}}
  MaterialExpressionConstant_5 {class=MaterialExpressionConstant, properties {r=90}}
  MaterialExpressionConstant_6 {class=MaterialExpressionConstant, properties {r=1024}}
  MaterialExpressionConstant_7 {class=MaterialExpressionConstant, properties {r=0.5}}
  MaterialExpressionConstant_8 {class=MaterialExpressionConstant, properties {r=0.5}}
  MaterialExpressionConstant_9 {class=MaterialExpressionConstant, properties {r=1}}
  MaterialExpressionDesaturation_1 {
   class=MaterialExpressionDesaturation
   inputs [
    None {from=MaterialExpressionReroute_32, output=""}
    Fraction {from=MaterialExpressionScalarParameter_49, output=""}
   ]
   properties {}
  }
  MaterialExpressionDesaturation_2 {
   class=MaterialExpressionDesaturation
   inputs [
    None {from=MaterialExpressionTextureSample_6, output=RGB}
    Fraction {from=MaterialExpressionScalarParameter_67, output=""}
   ]
   properties {}
  }
  MaterialExpressionDivide_0 {
   class=MaterialExpressionDivide
   inputs [
    A {from=MaterialExpressionAdd_10}
    B {from=MaterialExpressionNamedRerouteUsage_59}
   ]
  }
  MaterialExpressionDivide_1 {
   class=MaterialExpressionDivide
   inputs [
    A {from=MaterialExpressionSubtract_0}
    B {from=MaterialExpressionMax_3}
   ]
  }
  MaterialExpressionDivide_2 {
   class=MaterialExpressionDivide
   inputs [
    A {from=""}
    B {from=MaterialExpressionMaterialFunctionCall_48, output=Scale Z}
   ]
  }
  MaterialExpressionDivide_31 {
   class=MaterialExpressionDivide
   inputs [
    A {from=MaterialExpressionMaterialFunctionCall_5, output=R}
    B {from=MaterialExpressionMaterialFunctionCall_5, output=R}
   ]
   properties {desc=ratio}
  }
  MaterialExpressionDivide_32 {
   class=MaterialExpressionDivide
   inputs [
    A {from=MaterialExpressionMaterialFunctionCall_63, output=G}
    B {from=MaterialExpressionMaterialFunctionCall_63, output=G}
   ]
  }
  MaterialExpressionDivide_33 {
   class=MaterialExpressionDivide
   inputs [
    A {from=""}
    B {from=MaterialExpressionMultiply_2}
   ]
  }
  MaterialExpressionDivide_8 {
   class=MaterialExpressionDivide
   inputs [
    A {from=MaterialExpressionConstant_14}
    B {from=MaterialExpressionMultiply_38}
   ]
  }
  MaterialExpressionFresnel_0 {
   class=MaterialExpressionFresnel
   inputs [
    ExponentIn {from=MaterialExpressionScalarParameter_45, output=""}
    BaseReflectFractionIn {from="", output=""}
    Normal {from="", output=""}
   ]
   properties {baseReflectFraction=0}
  }
  MaterialExpressionFunctionInput_0 {
   class=MaterialExpressionFunctionInput
   inputs [Preview {from=MaterialExpressionVertexInterpolator_0, output=PS}]
   properties {bUsePreviewValueAsDefault=true, inputName=Pre-Skinned Local Normal, sortPriority=10}
  }
  MaterialExpressionFunctionInput_1 {
   class=MaterialExpressionFunctionInput
   inputs [Preview {from=MaterialExpressionTextureSample_3, output=G}]
   properties {bUsePreviewValueAsDefault=true, inputName=Mask Crevise, inputType=FunctionInput_Scalar, sortPriority=24, previewValue=<1 0 0 1>}
  }
  MaterialExpressionFunctionInput_10 {
   class=MaterialExpressionFunctionInput
   inputs [Preview {from=MaterialExpressionConstant2Vector_3, output=""}]
   properties {bUsePreviewValueAsDefault=true, inputName=IsState_B, inputType=FunctionInput_Vector2, sortPriority=4}
  }
  MaterialExpressionFunctionInput_11 {
   class=MaterialExpressionFunctionInput
   inputs [Preview {from=MaterialExpressionTextureSample_3, output=B}]
   properties {bUsePreviewValueAsDefault=true, inputName=Mask Secondary, inputType=FunctionInput_Scalar, sortPriority=24, previewValue=<1 0 0 1>}
  }
  MaterialExpressionFunctionInput_18 {
   class=MaterialExpressionFunctionInput
   inputs [Preview {from=MaterialExpressionTextureSample_3, output=R}]
   properties {bUsePreviewValueAsDefault=true, inputName=Mask Primary, inputType=FunctionInput_Scalar, sortPriority=24, previewValue=<1 0 0 1>}
  }
  MaterialExpressionFunctionInput_2 {
   class=MaterialExpressionFunctionInput
   inputs [Preview {from=MaterialExpressionDivide_1, output=""}]
   properties {bUsePreviewValueAsDefault=true, inputName=Calcated_Bounds, sortPriority=11}
  }
  MaterialExpressionFunctionInput_3 {
   class=MaterialExpressionFunctionInput
   inputs [Preview {from=MaterialExpressionConstant4Vector_0, output=RGBA}]
   properties {bUsePreviewValueAsDefault=true, inputName=IsState, inputType=FunctionInput_Vector4, sortPriority=4}
  }
  MaterialExpressionFunctionInput_30 {
   class=MaterialExpressionFunctionInput
   inputs [Preview {from=MaterialExpressionConstant4Vector_4, output=RGBA}]
   properties {bUsePreviewValueAsDefault=true, inputName=Object-SafeZone-Virtual-Pivot, inputType=FunctionInput_Vector4, sortPriority=105}
  }
  MaterialExpressionFunctionInput_4 {
   class=MaterialExpressionFunctionInput
   inputs [Preview {from=MaterialExpressionSetMaterialAttributes_1, output=""}]
   properties {inputName=Material_Root, inputType=FunctionInput_MaterialAttributes}
  }
  MaterialExpressionFunctionInput_5 {
   class=MaterialExpressionFunctionInput
   inputs [Preview {from="", output=""}]
   properties {inputName=TS_Emis_01_Msk, inputType=FunctionInput_Scalar}
  }
  MaterialExpressionFunctionInput_6 {
   class=MaterialExpressionFunctionInput
   inputs [Preview {from="", output=""}]
   properties {inputName=TS_Emis_02_Msk, inputType=FunctionInput_Scalar}
  }
  MaterialExpressionFunctionInput_7 {
   class=MaterialExpressionFunctionInput
   inputs [Preview {from=MaterialExpressionConstant_2, output=""}]
   properties {bUsePreviewValueAsDefault=true, inputName=Object-SafeZone-VP-Character-PreView-Switch, inputType=FunctionInput_Scalar, sortPriority=105}
  }
  MaterialExpressionFunctionInput_9 {
   class=MaterialExpressionFunctionInput
   inputs [Preview {from=MaterialExpressionSetMaterialAttributes_1, output=""}]
   properties {inputName=Material, inputType=FunctionInput_MaterialAttributes}
  }
  MaterialExpressionFunctionOutput_0 {class=MaterialExpressionFunctionOutput, inputs [None {from=MaterialExpressionStaticSwitchParameter_33, output=""}]}
  MaterialExpressionGetMaterialAttributes_0 {
   class=MaterialExpressionGetMaterialAttributes
   inputs [None {from=MaterialExpressionFunctionInput_9, output=""}]
   properties {attributeGetTypes [
     #1
     #2
     #3
     #4
    ]}
  }
  MaterialExpressionGetMaterialAttributes_1 {
   class=MaterialExpressionGetMaterialAttributes
   inputs [None {from=MaterialExpressionNamedRerouteUsage_97, output=""}]
   properties {attributeGetTypes [#1, #5]}
  }
  MaterialExpressionInverseLinearInterpolate_0 {
   class=MaterialExpressionInverseLinearInterpolate
   inputs [
    A {from=MaterialExpressionConstant_32, output=""}
    B {from=MaterialExpressionConstant_33, output=""}
    Value {from=MaterialExpressionMaterialFunctionCall_33, output=FOV angle in degrees}
   ]
   properties {}
  }
  MaterialExpressionInverseLinearInterpolate_1 {
   class=MaterialExpressionInverseLinearInterpolate
   inputs [
    A {from=MaterialExpressionConstant_3, output=""}
    B {from=MaterialExpressionConstant_5, output=""}
    Value {from=MaterialExpressionMaterialFunctionCall_40, output=FOV angle in degrees}
   ]
   properties {}
  }
  MaterialExpressionLinearInterpolate_0 {
   class=MaterialExpressionLinearInterpolate
   inputs [
    A {from=MaterialExpressionReroute_0}
    B {from=MaterialExpressionMultiply_6}
    Alpha {from=MaterialExpressionClamp_0}
   ]
  }
  MaterialExpressionLinearInterpolate_1 {
   class=MaterialExpressionLinearInterpolate
   inputs [
    A {from=MaterialExpressionLinearInterpolate_2}
    B {from=MaterialExpressionScalarParameter_7}
    Alpha {from=MaterialExpressionMaterialFunctionCall_17, output=G}
   ]
  }
  MaterialExpressionLinearInterpolate_10 {
   class=MaterialExpressionLinearInterpolate
   inputs [
    A {from=MaterialExpressionScalarParameter_42}
    B {from=MaterialExpressionScalarParameter_8}
    Alpha {from=MaterialExpressionMaterialFunctionCall_16, output=A}
   ]
  }
  MaterialExpressionLinearInterpolate_100 {
   class=MaterialExpressionLinearInterpolate
   inputs [
    A {from=MaterialExpressionMaterialFunctionCall_43, output=R}
    B {from=MaterialExpressionLinearInterpolate_99}
    Alpha {from=MaterialExpressionMaterialFunctionCall_61, output=R}
   ]
  }
  MaterialExpressionLinearInterpolate_101 {
   class=MaterialExpressionLinearInterpolate
   inputs [
    A {from=MaterialExpressionLinearInterpolate_100}
    B {from=MaterialExpressionMaterialFunctionCall_43, output=G}
    Alpha {from=MaterialExpressionMaterialFunctionCall_61, output=B}
   ]
  }
  MaterialExpressionLinearInterpolate_11 {
   class=MaterialExpressionLinearInterpolate
   inputs [
    A {from=MaterialExpressionScalarParameter_61}
    B {from=MaterialExpressionScalarParameter_73}
    Alpha {from=MaterialExpressionMaterialFunctionCall_35, output=G}
   ]
  }
  MaterialExpressionLinearInterpolate_12 {
   class=MaterialExpressionLinearInterpolate
   inputs [
    A {from=MaterialExpressionLinearInterpolate_49}
    B {from=MaterialExpressionVectorParameter_15, output=RGB}
    Alpha {from=MaterialExpressionMaterialFunctionCall_18, output=G}
   ]
  }
  MaterialExpressionLinearInterpolate_13 {
   class=MaterialExpressionLinearInterpolate
   inputs [
    A {from=MaterialExpressionMultiply_28}
    B {from=MaterialExpressionNamedRerouteUsage_55}
    Alpha {from=MaterialExpressionNamedRerouteUsage_70}
   ]
  }
  MaterialExpressionLinearInterpolate_14 {
   class=MaterialExpressionLinearInterpolate
   inputs [
    A {from=MaterialExpressionScalarParameter_22}
    B {from=MaterialExpressionScalarParameter_25}
    Alpha {from=MaterialExpressionMaterialFunctionCall_2, output=A}
   ]
  }
  MaterialExpressionLinearInterpolate_15 {
   class=MaterialExpressionLinearInterpolate
   inputs [
    A {from=MaterialExpressionLinearInterpolate_6}
    B {from=MaterialExpressionScalarParameter_23}
    Alpha {from=MaterialExpressionMaterialFunctionCall_3, output=R}
   ]
  }
  MaterialExpressionLinearInterpolate_16 {
   class=MaterialExpressionLinearInterpolate
   inputs [
    A {from=MaterialExpressionLinearInterpolate_15}
    B {from=MaterialExpressionScalarParameter_24}
    Alpha {from=MaterialExpressionMaterialFunctionCall_3, output=G}
   ]
  }
  MaterialExpressionLinearInterpolate_17 {
   class=MaterialExpressionLinearInterpolate
   inputs [
    A {from=MaterialExpressionAppendVector_9}
    B {from=MaterialExpressionAppendVector_10}
    Alpha {from=MaterialExpressionMaterialFunctionCall_21, output=A}
   ]
  }
  MaterialExpressionLinearInterpolate_18 {
   class=MaterialExpressionLinearInterpolate
   inputs [
    A {from=MaterialExpressionLinearInterpolate_35}
    B {from=MaterialExpressionVectorParameter_34, output=RGB}
    Alpha {from=MaterialExpressionSaturate_5}
   ]
  }
  MaterialExpressionLinearInterpolate_19 {
   class=MaterialExpressionLinearInterpolate
   inputs [
    A {from=MaterialExpressionLinearInterpolate_47}
    B {from=MaterialExpressionAppendVector_7}
    Alpha {from=MaterialExpressionMaterialFunctionCall_22, output=G}
   ]
  }
  MaterialExpressionLinearInterpolate_2 {
   class=MaterialExpressionLinearInterpolate
   inputs [
    A {from=MaterialExpressionLinearInterpolate_39}
    B {from=MaterialExpressionScalarParameter_1}
    Alpha {from=MaterialExpressionMaterialFunctionCall_17, output=R}
   ]
  }
  MaterialExpressionLinearInterpolate_20 {
   class=MaterialExpressionLinearInterpolate
   inputs [
    A {from=MaterialExpressionScalarParameter_90}
    B {from=MaterialExpressionScalarParameter_31}
    Alpha {from=MaterialExpressionSaturate_4}
   ]
  }
  MaterialExpressionLinearInterpolate_21 {
   class=MaterialExpressionLinearInterpolate
   inputs [
    A {from=MaterialExpressionVectorParameter_4, output=RGB}
    B {from=MaterialExpressionLinearInterpolate_23}
    Alpha {from=MaterialExpressionMaterialFunctionCall_46, output=R}
   ]
  }
  MaterialExpressionLinearInterpolate_22 {
   class=MaterialExpressionLinearInterpolate
   inputs [
    A {from=MaterialExpressionTransformPosition_0}
    B {from=MaterialExpressionComponentMask_9}
    Alpha {from=MaterialExpressionReroute_2}
   ]
  }
  MaterialExpressionLinearInterpolate_23 {
   class=MaterialExpressionLinearInterpolate
   inputs [
    A {from=MaterialExpressionVectorParameter_26, output=RGB}
    B {from=MaterialExpressionVectorParameter_58, output=RGB}
    Alpha {from=MaterialExpressionMaterialFunctionCall_46, output=G}
   ]
  }
  MaterialExpressionLinearInterpolate_24 {
   class=MaterialExpressionLinearInterpolate
   inputs [
    A {from=MaterialExpressionLinearInterpolate_21}
    B {from=MaterialExpressionVectorParameter_27, output=RGB}
    Alpha {from=MaterialExpressionMaterialFunctionCall_46, output=B}
   ]
  }
  MaterialExpressionLinearInterpolate_25 {
   class=MaterialExpressionLinearInterpolate
   inputs [
    A {from=MaterialExpressionScalarParameter_32}
    B {from=MaterialExpressionScalarParameter_33}
    Alpha {from=MaterialExpressionMaterialFunctionCall_9, output=G}
   ]
  }
  MaterialExpressionLinearInterpolate_26 {
   class=MaterialExpressionLinearInterpolate
   inputs [
    A {from=MaterialExpressionScalarParameter_34}
    B {from=MaterialExpressionLinearInterpolate_25}
    Alpha {from=MaterialExpressionMaterialFunctionCall_9, output=R}
   ]
  }
  MaterialExpressionLinearInterpolate_27 {
   class=MaterialExpressionLinearInterpolate
   inputs [
    A {from=MaterialExpressionLinearInterpolate_26}
    B {from=MaterialExpressionLinearInterpolate_28}
    Alpha {from=MaterialExpressionMaterialFunctionCall_9, output=B}
   ]
  }
  MaterialExpressionLinearInterpolate_28 {
   class=MaterialExpressionLinearInterpolate
   inputs [
    A {from=MaterialExpressionScalarParameter_35}
    B {from=MaterialExpressionScalarParameter_38}
    Alpha {from=MaterialExpressionMaterialFunctionCall_9, output=A}
   ]
  }
  MaterialExpressionLinearInterpolate_29 {
   class=MaterialExpressionLinearInterpolate
   inputs [
    A {from=MaterialExpressionLinearInterpolate_27}
    B {from=MaterialExpressionScalarParameter_36}
    Alpha {from=MaterialExpressionMaterialFunctionCall_10, output=R}
   ]
  }
  MaterialExpressionLinearInterpolate_3 {
   class=MaterialExpressionLinearInterpolate
   inputs [
    A {from=MaterialExpressionLinearInterpolate_33}
    B {from=MaterialExpressionLinearInterpolate_48}
    Alpha {from=MaterialExpressionMaterialFunctionCall_0, output=B}
   ]
  }
  MaterialExpressionLinearInterpolate_30 {
   class=MaterialExpressionLinearInterpolate
   inputs [
    A {from=MaterialExpressionLinearInterpolate_29}
    B {from=MaterialExpressionScalarParameter_37}
    Alpha {from=MaterialExpressionMaterialFunctionCall_10, output=G}
   ]
  }
  MaterialExpressionLinearInterpolate_31 {
   class=MaterialExpressionLinearInterpolate
   inputs [
    A {from=MaterialExpressionScalarParameter_74}
    B {from=MaterialExpressionLinearInterpolate_11}
    Alpha {from=MaterialExpressionMaterialFunctionCall_35, output=R}
   ]
  }
  MaterialExpressionLinearInterpolate_32 {
   class=MaterialExpressionLinearInterpolate
   inputs [
    A {from=MaterialExpressionVectorParameter_12, output=RGB}
    B {from=MaterialExpressionVectorParameter_2, output=RGB}
    Alpha {from=MaterialExpressionMaterialFunctionCall_0, output=G}
   ]
  }
  MaterialExpressionLinearInterpolate_33 {
   class=MaterialExpressionLinearInterpolate
   inputs [
    A {from=MaterialExpressionVectorParameter_1, output=RGB}
    B {from=MaterialExpressionLinearInterpolate_32}
    Alpha {from=MaterialExpressionMaterialFunctionCall_0, output=R}
   ]
  }
  MaterialExpressionLinearInterpolate_34 {
   class=MaterialExpressionLinearInterpolate
   inputs [
    A {from=MaterialExpressionReroute_33}
    B {from=MaterialExpressionMultiply_32}
    Alpha {from=MaterialExpressionReroute_31}
   ]
  }
  MaterialExpressionLinearInterpolate_35 {
   class=MaterialExpressionLinearInterpolate
   inputs [
    A {from=MaterialExpressionLinearInterpolate_34}
    B {from=MaterialExpressionMultiply_33}
    Alpha {from=MaterialExpressionMultiply_10}
   ]
  }
  MaterialExpressionLinearInterpolate_36 {
   class=MaterialExpressionLinearInterpolate
   inputs [
    A {from=MaterialExpressionTime_8}
    B {from=MaterialExpressionMultiply_43}
    Alpha {from=MaterialExpressionScalarParameter_54}
   ]
  }
  MaterialExpressionLinearInterpolate_37 {
   class=MaterialExpressionLinearInterpolate
   inputs [
    A {from=MaterialExpressionScalarParameter_39}
    B {from=MaterialExpressionScalarParameter_40}
    Alpha {from=MaterialExpressionMaterialFunctionCall_16, output=G}
   ]
  }
  MaterialExpressionLinearInterpolate_38 {
   class=MaterialExpressionLinearInterpolate
   inputs [
    A {from=MaterialExpressionScalarParameter_41}
    B {from=MaterialExpressionLinearInterpolate_37}
    Alpha {from=MaterialExpressionMaterialFunctionCall_16, output=R}
   ]
  }
  MaterialExpressionLinearInterpolate_39 {
   class=MaterialExpressionLinearInterpolate
   inputs [
    A {from=MaterialExpressionLinearInterpolate_38}
    B {from=MaterialExpressionLinearInterpolate_10}
    Alpha {from=MaterialExpressionMaterialFunctionCall_16, output=B}
   ]
  }
  MaterialExpressionLinearInterpolate_4 {
   class=MaterialExpressionLinearInterpolate
   inputs [
    A {from=MaterialExpressionScalarParameter_13}
    B {from=MaterialExpressionScalarParameter_14}
    Alpha {from=MaterialExpressionMaterialFunctionCall_2, output=G}
   ]
  }
  MaterialExpressionLinearInterpolate_40 {
   class=MaterialExpressionLinearInterpolate
   inputs [
    A {from=MaterialExpressionLinearInterpolate_31}
    B {from=MaterialExpressionLinearInterpolate_41}
    Alpha {from=MaterialExpressionMaterialFunctionCall_35, output=B}
   ]
  }
  MaterialExpressionLinearInterpolate_41 {
   class=MaterialExpressionLinearInterpolate
   inputs [
    A {from=MaterialExpressionScalarParameter_75}
    B {from=MaterialExpressionScalarParameter_78}
    Alpha {from=MaterialExpressionMaterialFunctionCall_35, output=A}
   ]
  }
  MaterialExpressionLinearInterpolate_42 {
   class=MaterialExpressionLinearInterpolate
   inputs [
    A {from=MaterialExpressionLinearInterpolate_40}
    B {from=MaterialExpressionScalarParameter_76}
    Alpha {from=MaterialExpressionMaterialFunctionCall_36, output=R}
   ]
  }
  MaterialExpressionLinearInterpolate_43 {
   class=MaterialExpressionLinearInterpolate
   inputs [
    A {from=MaterialExpressionLinearInterpolate_42}
    B {from=MaterialExpressionScalarParameter_77}
    Alpha {from=MaterialExpressionMaterialFunctionCall_36, output=G}
   ]
  }
  MaterialExpressionLinearInterpolate_44 {
   class=MaterialExpressionLinearInterpolate
   inputs [
    A {from=MaterialExpressionAppendVector_2}
    B {from=MaterialExpressionAppendVector_13}
    Alpha {from=MaterialExpressionMaterialFunctionCall_21, output=G}
   ]
  }
  MaterialExpressionLinearInterpolate_45 {
   class=MaterialExpressionLinearInterpolate
   inputs [
    A {from=MaterialExpressionAppendVector_1}
    B {from=MaterialExpressionLinearInterpolate_44}
    Alpha {from=MaterialExpressionMaterialFunctionCall_21, output=R}
   ]
  }
  MaterialExpressionLinearInterpolate_46 {
   class=MaterialExpressionLinearInterpolate
   inputs [
    A {from=MaterialExpressionLinearInterpolate_45}
    B {from=MaterialExpressionLinearInterpolate_17}
    Alpha {from=MaterialExpressionMaterialFunctionCall_21, output=B}
   ]
  }
  MaterialExpressionLinearInterpolate_47 {
   class=MaterialExpressionLinearInterpolate
   inputs [
    A {from=MaterialExpressionLinearInterpolate_46}
    B {from=MaterialExpressionAppendVector_8}
    Alpha {from=MaterialExpressionMaterialFunctionCall_22, output=R}
   ]
  }
  MaterialExpressionLinearInterpolate_48 {
   class=MaterialExpressionLinearInterpolate
   inputs [
    A {from=MaterialExpressionVectorParameter_11, output=RGB}
    B {from=MaterialExpressionVectorParameter_14, output=RGB}
    Alpha {from=MaterialExpressionMaterialFunctionCall_0, output=A}
   ]
  }
  MaterialExpressionLinearInterpolate_49 {
   class=MaterialExpressionLinearInterpolate
   inputs [
    A {from=MaterialExpressionLinearInterpolate_3}
    B {from=MaterialExpressionVectorParameter_36, output=RGB}
    Alpha {from=MaterialExpressionMaterialFunctionCall_18, output=R}
   ]
  }
  MaterialExpressionLinearInterpolate_5 {
   class=MaterialExpressionLinearInterpolate
   inputs [
    A {from=MaterialExpressionScalarParameter_21}
    B {from=MaterialExpressionLinearInterpolate_4}
    Alpha {from=MaterialExpressionMaterialFunctionCall_2, output=R}
   ]
  }
  MaterialExpressionLinearInterpolate_50 {
   class=MaterialExpressionLinearInterpolate
   inputs [
    A {from=MaterialExpressionScalarParameter_56}
    B {from=MaterialExpressionScalarParameter_57}
    Alpha {from=MaterialExpressionMaterialFunctionCall_23, output=G}
   ]
  }
  MaterialExpressionLinearInterpolate_51 {
   class=MaterialExpressionLinearInterpolate
   inputs [
    A {from=MaterialExpressionScalarParameter_58}
    B {from=MaterialExpressionLinearInterpolate_50}
    Alpha {from=MaterialExpressionMaterialFunctionCall_23, output=R}
   ]
  }
  MaterialExpressionLinearInterpolate_52 {
   class=MaterialExpressionLinearInterpolate
   inputs [
    A {from=MaterialExpressionLinearInterpolate_51}
    B {from=MaterialExpressionScalarParameter_59}
    Alpha {from=MaterialExpressionMaterialFunctionCall_23, output=B}
   ]
  }
  MaterialExpressionLinearInterpolate_53 {
   class=MaterialExpressionLinearInterpolate
   inputs [
    A {from=MaterialExpressionVectorParameter_28, output=RGB}
    B {from=MaterialExpressionVectorParameter_29, output=RGB}
    Alpha {from=MaterialExpressionPower_1}
   ]
  }
  MaterialExpressionLinearInterpolate_54 {
   class=MaterialExpressionLinearInterpolate
   inputs [
    A {from=MaterialExpressionGetMaterialAttributes_1, output=BaseColor}
    B {from=MaterialExpressionMultiply_16}
    Alpha {from=MaterialExpressionSaturate_9}
   ]
  }
  MaterialExpressionLinearInterpolate_55 {
   class=MaterialExpressionLinearInterpolate
   inputs [
    A {from=MaterialExpressionVectorParameter_0, output=RGB}
    B {from=MaterialExpressionMultiply_17}
    Alpha {from=MaterialExpressionSaturate_10}
   ]
  }
  MaterialExpressionLinearInterpolate_56 {
   class=MaterialExpressionLinearInterpolate
   inputs [
    A {from=MaterialExpressionMultiply_48}
    B {from=MaterialExpressionMultiply_49}
    Alpha {from=MaterialExpressionMaterialFunctionCall_25, output=G}
   ]
  }
  MaterialExpressionLinearInterpolate_57 {
   class=MaterialExpressionLinearInterpolate
   inputs [
    A {from=MaterialExpressionMultiply_47}
    B {from=MaterialExpressionLinearInterpolate_56}
    Alpha {from=MaterialExpressionMaterialFunctionCall_25, output=R}
   ]
  }
  MaterialExpressionLinearInterpolate_58 {
   class=MaterialExpressionLinearInterpolate
   inputs [
    A {from=MaterialExpressionLinearInterpolate_57}
    B {from=MaterialExpressionLinearInterpolate_59}
    Alpha {from=MaterialExpressionMaterialFunctionCall_25, output=B}
   ]
  }
  MaterialExpressionLinearInterpolate_59 {
   class=MaterialExpressionLinearInterpolate
   inputs [
    A {from=MaterialExpressionMultiply_51}
    B {from=MaterialExpressionMultiply_50}
    Alpha {from=MaterialExpressionMaterialFunctionCall_25, output=A}
   ]
  }
  MaterialExpressionLinearInterpolate_6 {
   class=MaterialExpressionLinearInterpolate
   inputs [
    A {from=MaterialExpressionLinearInterpolate_5}
    B {from=MaterialExpressionLinearInterpolate_14}
    Alpha {from=MaterialExpressionMaterialFunctionCall_2, output=B}
   ]
  }
  MaterialExpressionLinearInterpolate_60 {
   class=MaterialExpressionLinearInterpolate
   inputs [
    A {from=MaterialExpressionLinearInterpolate_58}
    B {from=MaterialExpressionMultiply_52}
    Alpha {from=MaterialExpressionMaterialFunctionCall_27, output=R}
   ]
  }
  MaterialExpressionLinearInterpolate_61 {
   class=MaterialExpressionLinearInterpolate
   inputs [
    A {from=MaterialExpressionLinearInterpolate_60}
    B {from=MaterialExpressionMultiply_53}
    Alpha {from=MaterialExpressionMaterialFunctionCall_27, output=G}
   ]
  }
  MaterialExpressionLinearInterpolate_62 {
   class=MaterialExpressionLinearInterpolate
   inputs [
    A {from=MaterialExpressionScalarParameter_81}
    B {from=MaterialExpressionScalarParameter_82}
    Alpha {from=MaterialExpressionMaterialFunctionCall_51, output=G}
   ]
  }
  MaterialExpressionLinearInterpolate_63 {
   class=MaterialExpressionLinearInterpolate
   inputs [
    A {from=MaterialExpressionInverseLinearInterpolate_1}
    B {from=""}
    Alpha {from=MaterialExpressionMaterialFunctionCall_38, output=R}
   ]
  }
  MaterialExpressionLinearInterpolate_64 {
   class=MaterialExpressionLinearInterpolate
   inputs [
    A {from=MaterialExpressionLinearInterpolate_63}
    B {from=MaterialExpressionLinearInterpolate_65}
    Alpha {from=MaterialExpressionMaterialFunctionCall_38, output=B}
   ]
  }
  MaterialExpressionLinearInterpolate_65 {
   class=MaterialExpressionLinearInterpolate
   inputs [
    A {from=""}
    B {from=MaterialExpressionReroute_5}
    Alpha {from=MaterialExpressionMaterialFunctionCall_38, output=A}
   ]
   properties {constA=1}
  }
  MaterialExpressionLinearInterpolate_66 {
   class=MaterialExpressionLinearInterpolate
   inputs [
    A {from=MaterialExpressionLinearInterpolate_64}
    B {from=MaterialExpressionReroute_4}
    Alpha {from=MaterialExpressionMaterialFunctionCall_39, output=R}
   ]
  }
  MaterialExpressionLinearInterpolate_67 {
   class=MaterialExpressionLinearInterpolate
   inputs [
    A {from=MaterialExpressionLinearInterpolate_66}
    B {from=MaterialExpressionReroute_4}
    Alpha {from=MaterialExpressionMaterialFunctionCall_39, output=G}
   ]
  }
  MaterialExpressionLinearInterpolate_68 {
   class=MaterialExpressionLinearInterpolate
   inputs [
    A {from=MaterialExpressionScalarParameter_86}
    B {from=MaterialExpressionScalarParameter_87}
    Alpha {from=MaterialExpressionMaterialFunctionCall_41, output=R}
   ]
  }
  MaterialExpressionLinearInterpolate_69 {
   class=MaterialExpressionLinearInterpolate
   inputs [
    A {from=MaterialExpressionLinearInterpolate_68}
    B {from=MaterialExpressionLinearInterpolate_70}
    Alpha {from=MaterialExpressionMaterialFunctionCall_41, output=B}
   ]
  }
  MaterialExpressionLinearInterpolate_7 {
   class=MaterialExpressionLinearInterpolate
   inputs [
    A {from=MaterialExpressionScalarParameter_15}
    B {from=MaterialExpressionScalarParameter_16}
    Alpha {from=MaterialExpressionMaterialFunctionCall_11, output=G}
   ]
  }
  MaterialExpressionLinearInterpolate_70 {
   class=MaterialExpressionLinearInterpolate
   inputs [
    A {from=MaterialExpressionScalarParameter_87}
    B {from=MaterialExpressionReroute_7}
    Alpha {from=MaterialExpressionMaterialFunctionCall_41, output=A}
   ]
   properties {constA=1}
  }
  MaterialExpressionLinearInterpolate_71 {
   class=MaterialExpressionLinearInterpolate
   inputs [
    A {from=MaterialExpressionLinearInterpolate_69}
    B {from=MaterialExpressionReroute_6}
    Alpha {from=MaterialExpressionMaterialFunctionCall_42, output=R}
   ]
  }
  MaterialExpressionLinearInterpolate_72 {
   class=MaterialExpressionLinearInterpolate
   inputs [
    A {from=MaterialExpressionLinearInterpolate_71}
    B {from=MaterialExpressionReroute_6}
    Alpha {from=MaterialExpressionMaterialFunctionCall_42, output=G}
   ]
  }
  MaterialExpressionLinearInterpolate_73 {
   class=MaterialExpressionLinearInterpolate
   inputs [
    A {from=MaterialExpressionScalarParameter_83}
    B {from=MaterialExpressionLinearInterpolate_62}
    Alpha {from=MaterialExpressionMaterialFunctionCall_51, output=R}
   ]
  }
  MaterialExpressionLinearInterpolate_74 {
   class=MaterialExpressionLinearInterpolate
   inputs [
    A {from=MaterialExpressionLinearInterpolate_73}
    B {from=MaterialExpressionLinearInterpolate_77}
    Alpha {from=MaterialExpressionMaterialFunctionCall_51, output=B}
   ]
  }
  MaterialExpressionLinearInterpolate_75 {
   class=MaterialExpressionLinearInterpolate
   inputs [
    A {from=MaterialExpressionLinearInterpolate_76}
    B {from=MaterialExpressionScalarParameter_88}
    Alpha {from=MaterialExpressionMaterialFunctionCall_52, output=G}
   ]
  }
  MaterialExpressionLinearInterpolate_76 {
   class=MaterialExpressionLinearInterpolate
   inputs [
    A {from=MaterialExpressionLinearInterpolate_74}
    B {from=MaterialExpressionScalarParameter_85}
    Alpha {from=MaterialExpressionMaterialFunctionCall_52, output=R}
   ]
  }
  MaterialExpressionLinearInterpolate_77 {
   class=MaterialExpressionLinearInterpolate
   inputs [
    A {from=MaterialExpressionScalarParameter_84}
    B {from=MaterialExpressionScalarParameter_89}
    Alpha {from=MaterialExpressionMaterialFunctionCall_51, output=A}
   ]
  }
  MaterialExpressionLinearInterpolate_78 {
   class=MaterialExpressionLinearInterpolate
   inputs [
    A {from=MaterialExpressionFunctionInput_5}
    B {from=MaterialExpressionFunctionInput_6}
    Alpha {from=MaterialExpressionScalarParameter_92}
   ]
  }
  MaterialExpressionLinearInterpolate_79 {
   class=MaterialExpressionLinearInterpolate
   inputs [
    A {from=MaterialExpressionNamedRerouteUsage_69}
    B {from=MaterialExpressionLinearInterpolate_78}
    Alpha {from=MaterialExpressionSaturate_15}
   ]
  }
  MaterialExpressionLinearInterpolate_8 {
   class=MaterialExpressionLinearInterpolate
   inputs [
    A {from=MaterialExpressionScalarParameter_17}
    B {from=MaterialExpressionLinearInterpolate_7}
    Alpha {from=MaterialExpressionMaterialFunctionCall_11, output=R}
   ]
  }
  MaterialExpressionLinearInterpolate_80 {
   class=MaterialExpressionLinearInterpolate
   inputs [
    A {from=""}
    B {from=MaterialExpressionNamedRerouteUsage_106}
    Alpha {from=MaterialExpressionScalarParameter_93}
   ]
  }
  MaterialExpressionLinearInterpolate_9 {
   class=MaterialExpressionLinearInterpolate
   inputs [
    A {from=MaterialExpressionLinearInterpolate_8}
    B {from=MaterialExpressionScalarParameter_18}
    Alpha {from=MaterialExpressionMaterialFunctionCall_11, output=B}
   ]
  }
  MaterialExpressionLinearInterpolate_99 {
   class=MaterialExpressionLinearInterpolate
   inputs [
    A {from=MaterialExpressionMaterialFunctionCall_43, output=B}
    B {from=MaterialExpressionMaterialFunctionCall_43, output=B}
    Alpha {from=MaterialExpressionMaterialFunctionCall_61, output=G}
   ]
  }
  MaterialExpressionMaterialFunctionCall_0 {
   class=MaterialExpressionMaterialFunctionCall
   inputs [Float4 {from=MaterialExpressionNamedRerouteUsage_57}]
   properties {materialFunction=&BreakOutFloat4Components}
  }
  MaterialExpressionMaterialFunctionCall_1 {
   class=MaterialExpressionMaterialFunctionCall
   inputs [Float2 {from=MaterialExpressionNamedRerouteUsage_26}]
   properties {materialFunction=&BreakOutFloat2Components}
  }
  MaterialExpressionMaterialFunctionCall_10 {
   class=MaterialExpressionMaterialFunctionCall
   inputs [Float2 {from=MaterialExpressionNamedRerouteUsage_49}]
   properties {materialFunction=&BreakOutFloat2Components}
  }
  MaterialExpressionMaterialFunctionCall_11 {
   class=MaterialExpressionMaterialFunctionCall
   inputs [Float4 {from=MaterialExpressionNamedRerouteUsage_25}]
   properties {materialFunction=&BreakOutFloat4Components}
  }
  MaterialExpressionMaterialFunctionCall_12 {
   class=MaterialExpressionMaterialFunctionCall
   inputs [
    CameraVector {from=""}
    Texture {from=MaterialExpressionNamedRerouteUsage_11}
    Tiling {from=MaterialExpressionScalarParameter_47}
    Offset {from=""}
    AxisFadeContrast {from=""}
   ]
   properties {materialFunction=/Engine/Functions/Engine_MaterialFunctions01/Texturing/TriplanarCameraVector}
  }
  MaterialExpressionMaterialFunctionCall_13 {
   class=MaterialExpressionMaterialFunctionCall
   inputs [
    CameraVector {from=""}
    Rotation Speed {from=MaterialExpressionMultiply_27}
    Texture {from=MaterialExpressionNamedRerouteUsage_10}
    Tiling {from=MaterialExpressionScalarParameter_46}
    Offset {from=MaterialExpressionMultiply_36}
    AxisFadeContrast {from=""}
   ]
   properties {materialFunction=/Game/Weapons/WeaponSkins/MaterialFunction/TriplanarCameraVector_Rotated}
  }
  MaterialExpressionMaterialFunctionCall_14 {
   class=MaterialExpressionMaterialFunctionCall
   inputs [
    CameraVector {from=""}
    Rotation Speed {from=MaterialExpressionReroute_35}
    Texture {from=MaterialExpressionReroute_36}
    Tiling {from=MaterialExpressionScalarParameter_50}
    Offset {from=""}
    AxisFadeContrast {from=MaterialExpressionScalarParameter_62}
   ]
   properties {materialFunction=&TriplanarCameraVector_Rotated}
  }
  MaterialExpressionMaterialFunctionCall_15 {
   class=MaterialExpressionMaterialFunctionCall
   inputs [
    Normal_Map {from=MaterialExpressionNamedRerouteUsage_56}
    Offset Rt {from=""}
    Pivot {from=""}
    Caculated_Bounds {from=MaterialExpressionNamedRerouteUsage_50}
    Pre_Skinned_Normals {from=MaterialExpressionNamedRerouteUsage_51}
    is_State {from=MaterialExpressionNamedRerouteUsage_52}
    UV_Rotation {from=MaterialExpressionNamedRerouteUsage_53}
    Uvs_Offset {from=""}
    Uvs_TileRate {from=""}
    Uvs_Offset_Pivot {from=""}
    For_Safe_Zones {from=""}
    Rotation_Comp {from=""}
    Clamp Y {from=""}
    Clamp X {from=""}
    Clamp Padding_Comp {from=""}
    No_Rotation {from=MaterialExpressionStaticBoolParameter_0}
   ]
   properties {materialFunction=&/MF_Projection_Side_Normal_V2}
  }
  MaterialExpressionMaterialFunctionCall_16 {
   class=MaterialExpressionMaterialFunctionCall
   inputs [Float4 {from=MaterialExpressionNamedRerouteUsage_60}]
   properties {materialFunction=&BreakOutFloat4Components}
  }
  MaterialExpressionMaterialFunctionCall_17 {
   class=MaterialExpressionMaterialFunctionCall
   inputs [Float2 {from=MaterialExpressionNamedRerouteUsage_61}]
   properties {materialFunction=&BreakOutFloat2Components}
  }
  MaterialExpressionMaterialFunctionCall_18 {
   class=MaterialExpressionMaterialFunctionCall
   inputs [Float2 {from=MaterialExpressionNamedRerouteUsage_58}]
   properties {materialFunction=&BreakOutFloat2Components}
  }
  MaterialExpressionMaterialFunctionCall_19 {
   class=MaterialExpressionMaterialFunctionCall
   inputs [
    A {from=MaterialExpressionVectorParameter_30, output=RGB}
    B {from=MaterialExpressionVectorParameter_16, output=RGB}
    C {from=MaterialExpressionVectorParameter_7, output=RGB}
    Alpha {from=MaterialExpressionPower_0}
   ]
   properties {materialFunction=&3ColorBlend}
  }
  MaterialExpressionMaterialFunctionCall_2 {
   class=MaterialExpressionMaterialFunctionCall
   inputs [Float4 {from=MaterialExpressionNamedRerouteUsage_22}]
   properties {materialFunction=&BreakOutFloat4Components}
  }
  MaterialExpressionMaterialFunctionCall_20 {
   class=MaterialExpressionMaterialFunctionCall
   inputs [
    Normal_Map {from=""}
    Offset Rt {from=""}
    Pivot {from=""}
    Caculated_Bounds {from=MaterialExpressionNamedRerouteUsage_72}
    Pre_Skinned_Normals {from=MaterialExpressionNamedRerouteUsage_73}
    is_State {from=MaterialExpressionNamedRerouteUsage_74}
    UV_Rotation {from=MaterialExpressionNamedRerouteUsage_80}
    Uvs_Offset {from=MaterialExpressionComponentMask_1}
    Uvs_TileRate {from=MaterialExpressionComponentMask_2}
    Uvs_Offset_Pivot {from=""}
    For_Safe_Zones {from=MaterialExpressionStaticBool_0}
    Rotation_Comp {from=""}
    Clamp Y {from=""}
    Clamp X {from=""}
    Clamp Padding_Comp {from=""}
    No_Rotation {from=""}
   ]
   properties {materialFunction=&MF_Projection_Side_Normal_V2}
  }
  MaterialExpressionMaterialFunctionCall_21 {
   class=MaterialExpressionMaterialFunctionCall
   inputs [Float4 {from=MaterialExpressionNamedRerouteUsage_75}]
   properties {materialFunction=&BreakOutFloat4Components}
  }
  MaterialExpressionMaterialFunctionCall_22 {
   class=MaterialExpressionMaterialFunctionCall
   inputs [Float2 {from=MaterialExpressionNamedRerouteUsage_76}]
   properties {materialFunction=&BreakOutFloat2Components}
  }
  MaterialExpressionMaterialFunctionCall_23 {
   class=MaterialExpressionMaterialFunctionCall
   inputs [Float4 {from=MaterialExpressionNamedRerouteUsage_77}]
   properties {materialFunction=&BreakOutFloat4Components}
  }
  MaterialExpressionMaterialFunctionCall_24 {
   class=MaterialExpressionMaterialFunctionCall
   inputs [Float2 {from=MaterialExpressionNamedRerouteUsage_78}]
   properties {materialFunction=&BreakOutFloat2Components}
  }
  MaterialExpressionMaterialFunctionCall_25 {
   class=MaterialExpressionMaterialFunctionCall
   inputs [Float4 {from=MaterialExpressionNamedRerouteUsage_89}]
   properties {materialFunction=&BreakOutFloat4Components}
  }
  MaterialExpressionMaterialFunctionCall_26 {
   class=MaterialExpressionMaterialFunctionCall
   inputs [Float2 {from=MaterialExpressionNamedRerouteUsage_36}]
   properties {materialFunction=&BreakOutFloat2Components}
  }
  MaterialExpressionMaterialFunctionCall_27 {
   class=MaterialExpressionMaterialFunctionCall
   inputs [Float2 {from=MaterialExpressionNamedRerouteUsage_90}]
   properties {materialFunction=&BreakOutFloat2Components}
  }
  MaterialExpressionMaterialFunctionCall_28 {
   class=MaterialExpressionMaterialFunctionCall
   inputs [
    In {from=MaterialExpressionChannelMaskParameter_0, output=R}
    Contrast {from=MaterialExpressionScalarParameter_26}
   ]
   properties {materialFunction=&CheapContrast}
  }
  MaterialExpressionMaterialFunctionCall_29 {
   class=MaterialExpressionMaterialFunctionCall
   inputs [
    Value_Morning {from=MaterialExpressionConstant_4}
    Value_Day {from=MaterialExpressionConstant_9}
    Value_Evening {from=MaterialExpressionConstant_0}
    Value_Night {from=MaterialExpressionConstant_12}
   ]
   properties {materialFunction=&Scalar_Time_of_Day}
  }
  MaterialExpressionMaterialFunctionCall_3 {
   class=MaterialExpressionMaterialFunctionCall
   inputs [Float2 {from=MaterialExpressionNamedRerouteUsage_23}]
   properties {materialFunction=&BreakOutFloat2Components}
  }
  MaterialExpressionMaterialFunctionCall_30 {
   class=MaterialExpressionMaterialFunctionCall
   inputs [
    A {from=MaterialExpressionVectorParameter_33, output=RGB}
    B {from=MaterialExpressionVectorParameter_32, output=RGB}
    C {from=MaterialExpressionVectorParameter_7, output=RGB}
    Alpha {from=MaterialExpressionSaturate_2}
   ]
   properties {materialFunction=&3ColorBlend}
  }
  MaterialExpressionMaterialFunctionCall_31 {
   class=MaterialExpressionMaterialFunctionCall
   inputs [
    Plane Normal Axis {from=MaterialExpressionNormalize_2}
    Plane X Axis {from=""}
    Plane Center {from=MaterialExpressionAdd_3}
    UV Scale {from=MaterialExpressionMultiply_42}
   ]
   properties {materialFunction=&VirtualPlaneCoordinates}
  }
  MaterialExpressionMaterialFunctionCall_32 {
   class=MaterialExpressionMaterialFunctionCall
   inputs []
   properties {materialFunction=&CameraDirectionVector}
  }
  MaterialExpressionMaterialFunctionCall_33 {
   class=MaterialExpressionMaterialFunctionCall
   inputs []
   properties {materialFunction=/Engine/Functions/Engine_MaterialFunctions02/Utility/FOV}
  }
  MaterialExpressionMaterialFunctionCall_34 {
   class=MaterialExpressionMaterialFunctionCall
   inputs [
    Texture {from=MaterialExpressionNamedRerouteUsage_1}
    Caculated_Bounds {from=MaterialExpressionNamedRerouteUsage_31}
    Pre_Skinned_Normals {from=MaterialExpressionNamedRerouteUsage_32}
    Blend_Contrast {from=MaterialExpressionScalarParameter_28}
    IsStatsB {from=MaterialExpressionNamedRerouteUsage_34}
    IsState {from=MaterialExpressionNamedRerouteUsage_33}
    Fill_alternate_Axis_Opacity {from=MaterialExpressionNamedRerouteUsage_94}
    "Fill_alternate_Axis_Color_A=Roughness" {from=MaterialExpressionAppendVector_0}
    Offset {from=MaterialExpressionAdd_18}
    Scale {from=MaterialExpressionNamedRerouteUsage_91}
    Offset_Pivot {from=""}
    Rotation_Axis {from=MaterialExpressionNamedRerouteUsage_0}
   ]
   properties {materialFunction=&/MF_TopsSideProjection_Base_V2}
  }
  MaterialExpressionMaterialFunctionCall_35 {
   class=MaterialExpressionMaterialFunctionCall
   inputs [Float4 {from=MaterialExpressionNamedRerouteUsage_47}]
   properties {materialFunction=&BreakOutFloat4Components}
  }
  MaterialExpressionMaterialFunctionCall_36 {
   class=MaterialExpressionMaterialFunctionCall
   inputs [Float2 {from=MaterialExpressionNamedRerouteUsage_54}]
   properties {materialFunction=&BreakOutFloat2Components}
  }
  MaterialExpressionMaterialFunctionCall_37 {
   class=MaterialExpressionMaterialFunctionCall
   inputs []
   properties {materialFunction=/Engine/Functions/Engine_MaterialFunctions02/WorldPositionOffset/ObjectScale}
  }
  MaterialExpressionMaterialFunctionCall_38 {
   class=MaterialExpressionMaterialFunctionCall
   inputs [Float4 {from=MaterialExpressionNamedRerouteUsage_63}]
   properties {materialFunction=&BreakOutFloat4Components}
  }
  MaterialExpressionMaterialFunctionCall_39 {
   class=MaterialExpressionMaterialFunctionCall
   inputs [Float2 {from=MaterialExpressionNamedRerouteUsage_66}]
   properties {materialFunction=&BreakOutFloat2Components}
  }
  MaterialExpressionMaterialFunctionCall_4 {
   class=MaterialExpressionMaterialFunctionCall
   inputs [
    Normal_Map {from=MaterialExpressionNamedRerouteUsage_38}
    Offset Rt {from=""}
    Pivot {from=""}
    Caculated_Bounds {from=MaterialExpressionNamedRerouteUsage_18}
    Pre_Skinned_Normals {from=MaterialExpressionNamedRerouteUsage_19}
    is_State {from=MaterialExpressionNamedRerouteUsage_21}
    UV_Rotation {from=MaterialExpressionNamedRerouteUsage_24}
    Uvs_Offset {from=""}
    Uvs_TileRate {from=MaterialExpressionNamedRerouteUsage_15}
    Uvs_Offset_Pivot {from=MaterialExpressionNamedRerouteUsage_27}
    For_Safe_Zones {from=""}
    Rotation_Comp {from=""}
    Clamp Y {from=""}
    Clamp X {from=""}
    Clamp Padding_Comp {from=""}
    No_Rotation {from=MaterialExpressionStaticBoolParameter_0}
   ]
   properties {materialFunction=&MF_Projection_Side_Normal_V2}
  }
  MaterialExpressionMaterialFunctionCall_40 {
   class=MaterialExpressionMaterialFunctionCall
   inputs []
   properties {materialFunction=&FOV}
  }
  MaterialExpressionMaterialFunctionCall_41 {
   class=MaterialExpressionMaterialFunctionCall
   inputs [Float4 {from=MaterialExpressionNamedRerouteUsage_92}]
   properties {materialFunction=&BreakOutFloat4Components}
  }
  MaterialExpressionMaterialFunctionCall_42 {
   class=MaterialExpressionMaterialFunctionCall
   inputs [Float2 {from=MaterialExpressionNamedRerouteUsage_93}]
   properties {materialFunction=&BreakOutFloat2Components}
  }
  MaterialExpressionMaterialFunctionCall_43 {
   class=MaterialExpressionMaterialFunctionCall
   inputs [Color {from=MaterialExpressionAbs_1}]
   properties {materialFunction=&SplitComponents}
  }
  MaterialExpressionMaterialFunctionCall_44 {
   class=MaterialExpressionMaterialFunctionCall
   inputs [
    In {from=MaterialExpressionLinearInterpolate_101}
    Contrast {from=MaterialExpressionScalarParameter_60}
   ]
   properties {materialFunction=&CheapContrast}
  }
  MaterialExpressionMaterialFunctionCall_45 {
   class=MaterialExpressionMaterialFunctionCall
   inputs [Float3 {from=MaterialExpressionSubtract_4}]
   properties {materialFunction=&BreakOutFloat3Components}
  }
  MaterialExpressionMaterialFunctionCall_46 {
   class=MaterialExpressionMaterialFunctionCall
   inputs [Float4 {from=MaterialExpressionNamedRerouteUsage_35}]
   properties {materialFunction=&BreakOutFloat4Components}
  }
  MaterialExpressionMaterialFunctionCall_47 {
   class=MaterialExpressionMaterialFunctionCall
   inputs [Center Point World Position {from=MaterialExpressionLinearInterpolate_22}]
   properties {materialFunction=/Engine/Functions/Engine_MaterialFunctions02/UVs/WorldSpaceAlignedScreenCoordinates}
  }
  MaterialExpressionMaterialFunctionCall_48 {
   class=MaterialExpressionMaterialFunctionCall
   inputs []
   properties {materialFunction=/Engine/Functions/Engine_MaterialFunctions02/WorldPositionOffset/V2/ObjectScale}
  }
  MaterialExpressionMaterialFunctionCall_49 {
   class=MaterialExpressionMaterialFunctionCall
   inputs [
    In {from=MaterialExpressionOneMinus_3}
    Contrast {from=MaterialExpressionConstant_7}
   ]
   properties {materialFunction=&CheapContrast}
  }
  MaterialExpressionMaterialFunctionCall_5 {
   class=MaterialExpressionMaterialFunctionCall
   inputs [Float2 {from=MaterialExpressionConstant_6}]
   properties {materialFunction=&BreakOutFloat2Components}
  }
  MaterialExpressionMaterialFunctionCall_50 {
   class=MaterialExpressionMaterialFunctionCall
   inputs [
    In {from=MaterialExpressionOneMinus_4}
    Contrast {from=MaterialExpressionScalarParameter_66}
   ]
   properties {materialFunction=&CheapContrast}
  }
  MaterialExpressionMaterialFunctionCall_51 {
   class=MaterialExpressionMaterialFunctionCall
   inputs [Float4 {from=MaterialExpressionNamedRerouteUsage_14}]
   properties {materialFunction=&BreakOutFloat4Components}
  }
  MaterialExpressionMaterialFunctionCall_52 {
   class=MaterialExpressionMaterialFunctionCall
   inputs [Float2 {from=MaterialExpressionNamedRerouteUsage_99}]
   properties {materialFunction=&BreakOutFloat2Components}
  }
  MaterialExpressionMaterialFunctionCall_59 {
   class=MaterialExpressionMaterialFunctionCall
   inputs [
    UVs {from=MaterialExpressionScreenPosition_3, output=ViewportUV}
    Texture Scale {from=MaterialExpressionAppendVector_3}
   ]
   properties {materialFunction=&ScaleUVsByCenter}
  }
  MaterialExpressionMaterialFunctionCall_6 {
   class=MaterialExpressionMaterialFunctionCall
   inputs [Color {from=MaterialExpressionSubtract_3}]
   properties {materialFunction=&SplitComponents}
  }
  MaterialExpressionMaterialFunctionCall_61 {
   class=MaterialExpressionMaterialFunctionCall
   inputs [Float4 {from=MaterialExpressionNamedRerouteUsage_82}]
   properties {materialFunction=&BreakOutFloat4Components}
  }
  MaterialExpressionMaterialFunctionCall_62 {
   class=MaterialExpressionMaterialFunctionCall
   inputs []
   properties {materialFunction=&ScreenResolution}
  }
  MaterialExpressionMaterialFunctionCall_63 {
   class=MaterialExpressionMaterialFunctionCall
   inputs [Float2 {from=MaterialExpressionMaterialFunctionCall_62, output=Visible Resolution}]
   properties {materialFunction=&BreakOutFloat2Components}
  }
  MaterialExpressionMaterialFunctionCall_7 {
   class=MaterialExpressionMaterialFunctionCall
   inputs [
    In {from=MaterialExpressionSaturate_1}
    Contrast {from=MaterialExpressionScalarParameter_19}
   ]
   properties {materialFunction=&CheapContrast}
  }
  MaterialExpressionMaterialFunctionCall_8 {
   class=MaterialExpressionMaterialFunctionCall
   inputs [
    In {from=MaterialExpressionMultiply_55}
    Contrast {from=MaterialExpressionScalarParameter_20}
   ]
   properties {materialFunction=&CheapContrast}
  }
  MaterialExpressionMaterialFunctionCall_9 {
   class=MaterialExpressionMaterialFunctionCall
   inputs [Float4 {from=MaterialExpressionNamedRerouteUsage_48}]
   properties {materialFunction=&BreakOutFloat4Components}
  }
  MaterialExpressionMax_0 {
   class=MaterialExpressionMax
   inputs [
    A {from=MaterialExpressionMaterialFunctionCall_45, output=R}
    B {from=MaterialExpressionMaterialFunctionCall_45, output=R}
   ]
  }
  MaterialExpressionMax_1 {
   class=MaterialExpressionMax
   inputs [
    A {from=MaterialExpressionScalarParameter_91}
    B {from=MaterialExpressionScalarParameter_92}
   ]
  }
  MaterialExpressionMax_2 {
   class=MaterialExpressionMax
   inputs [
    A {from=MaterialExpressionMaterialFunctionCall_6, output=R}
    B {from=MaterialExpressionMaterialFunctionCall_6, output=R}
   ]
  }
  MaterialExpressionMax_3 {
   class=MaterialExpressionMax
   inputs [
    A {from=MaterialExpressionMax_2}
    B {from=MaterialExpressionMaterialFunctionCall_6, output=B}
   ]
  }
  MaterialExpressionMax_5 {
   class=MaterialExpressionMax
   inputs [
    A {from=MaterialExpressionMax_0}
    B {from=MaterialExpressionMaterialFunctionCall_45, output=B}
   ]
  }
  MaterialExpressionMultiply_0 {
   class=MaterialExpressionMultiply
   inputs [
    A {from=MaterialExpressionTime_1}
    B {from=MaterialExpressionVectorParameter_6, output=RGB}
   ]
  }
  MaterialExpressionMultiply_1 {
   class=MaterialExpressionMultiply
   inputs [
    A {from=MaterialExpressionChannelMaskParameter_0, output=R}
    B {from=MaterialExpressionScalarParameter_12}
   ]
  }
  MaterialExpressionMultiply_10 {
   class=MaterialExpressionMultiply
   inputs [
    A {from=MaterialExpressionMultiply_7}
    B {from=MaterialExpressionTextureSample_10, output=G}
   ]
  }
  MaterialExpressionMultiply_11 {
   class=MaterialExpressionMultiply
   inputs [
    A {from=MaterialExpressionChannelMaskParameter_0, output=R}
    B {from=MaterialExpressionScalarParameter_55}
   ]
  }
  MaterialExpressionMultiply_12 {
   class=MaterialExpressionMultiply
   inputs [
    A {from=MaterialExpressionMultiply_13}
    B {from=MaterialExpressionLinearInterpolate_53}
   ]
  }
  MaterialExpressionMultiply_13 {
   class=MaterialExpressionMultiply
   inputs [
    A {from=MaterialExpressionNamedRerouteUsage_84}
    B {from=MaterialExpressionTextureSample_2, output=R}
   ]
  }
  MaterialExpressionMultiply_14 {
   class=MaterialExpressionMultiply
   inputs [
    A {from=MaterialExpressionAdd_14}
    B {from=MaterialExpressionNamedRerouteUsage_85}
   ]
  }
  MaterialExpressionMultiply_15 {
   class=MaterialExpressionMultiply
   inputs [
    A {from=MaterialExpressionAdd_16}
    B {from=""}
   ]
   properties {constB=2}
  }
  MaterialExpressionMultiply_16 {
   class=MaterialExpressionMultiply
   inputs [
    A {from=MaterialExpressionGetMaterialAttributes_1, output=BaseColor}
    B {from=MaterialExpressionScalarParameter_63}
   ]
  }
  MaterialExpressionMultiply_17 {
   class=MaterialExpressionMultiply
   inputs [
    A {from=MaterialExpressionVectorParameter_0, output=RGB}
    B {from=MaterialExpressionScalarParameter_64}
   ]
  }
  MaterialExpressionMultiply_18 {
   class=MaterialExpressionMultiply
   inputs [
    A {from=MaterialExpressionAdd_17}
    B {from=""}
   ]
   properties {constB=2}
  }
  MaterialExpressionMultiply_19 {
   class=MaterialExpressionMultiply
   inputs [
    A {from=MaterialExpressionMaterialFunctionCall_30, output=Result}
    B {from=""}
   ]
  }
  MaterialExpressionMultiply_2 {
   class=MaterialExpressionMultiply
   inputs [
    A {from=MaterialExpressionDivide_32}
    B {from=MaterialExpressionDivide_31}
   ]
  }
  MaterialExpressionMultiply_20 {
   class=MaterialExpressionMultiply
   inputs [
    A {from=MaterialExpressionMaterialFunctionCall_19, output=Result}
    B {from=MaterialExpressionPower_0}
   ]
   properties {constA=1}
  }
  MaterialExpressionMultiply_21 {
   class=MaterialExpressionMultiply
   inputs [
    A {from=MaterialExpressionNamedRerouteDeclaration_12}
    B {from=MaterialExpressionMultiply_20}
   ]
  }
  MaterialExpressionMultiply_22 {
   class=MaterialExpressionMultiply
   inputs [
    A {from=MaterialExpressionAdd_2}
    B {from=""}
   ]
   properties {constB=4}
  }
  MaterialExpressionMultiply_23 {
   class=MaterialExpressionMultiply
   inputs [
    A {from=""}
    B {from=MaterialExpressionScalarParameter_43}
   ]
  }
  MaterialExpressionMultiply_24 {
   class=MaterialExpressionMultiply
   inputs [
    A {from=MaterialExpressionMaterialFunctionCall_32, output=Result}
    B {from=""}
   ]
   properties {constB=50}
  }
  MaterialExpressionMultiply_25 {
   class=MaterialExpressionMultiply
   inputs [
    A {from=MaterialExpressionConstant3Vector_4}
    B {from=MaterialExpressionMaterialFunctionCall_32, output=Result}
   ]
  }
  MaterialExpressionMultiply_26 {
   class=MaterialExpressionMultiply
   inputs [
    A {from=""}
    B {from=MaterialExpressionNormalize_3}
   ]
   properties {constA=0.25}
  }
  MaterialExpressionMultiply_27 {
   class=MaterialExpressionMultiply
   inputs [
    A {from=MaterialExpressionTime_5}
    B {from=MaterialExpressionScalarParameter_48}
   ]
  }
  MaterialExpressionMultiply_28 {
   class=MaterialExpressionMultiply
   inputs [
    A {from=MaterialExpressionNamedRerouteUsage_55}
    B {from=""}
   ]
   properties {constB=0.5}
  }
  MaterialExpressionMultiply_29 {
   class=MaterialExpressionMultiply
   inputs [
    A {from=MaterialExpressionVectorParameter_34, output=A}
    B {from=MaterialExpressionFresnel_0}
   ]
  }
  MaterialExpressionMultiply_3 {
   class=MaterialExpressionMultiply
   inputs [
    A {from=MaterialExpressionMaterialFunctionCall_29, output=Result}
    B {from=""}
   ]
  }
  MaterialExpressionMultiply_30 {
   class=MaterialExpressionMultiply
   inputs [
    A {from=MaterialExpressionComponentMask_5}
    B {from=MaterialExpressionComponentMask_5}
   ]
  }
  MaterialExpressionMultiply_31 {
   class=MaterialExpressionMultiply
   inputs [
    A {from=MaterialExpressionComponentMask_6}
    B {from=MaterialExpressionComponentMask_6}
   ]
  }
  MaterialExpressionMultiply_32 {
   class=MaterialExpressionMultiply
   inputs [
    A {from=MaterialExpressionDesaturation_1}
    B {from=MaterialExpressionVectorParameter_8, output=RGB}
   ]
  }
  MaterialExpressionMultiply_33 {
   class=MaterialExpressionMultiply
   inputs [
    A {from=MaterialExpressionDesaturation_1}
    B {from=MaterialExpressionVectorParameter_9, output=RGB}
   ]
  }
  MaterialExpressionMultiply_34 {
   class=MaterialExpressionMultiply
   inputs [
    A {from=MaterialExpressionMultiply_31}
    B {from=MaterialExpressionTextureSample_8, output=G}
   ]
  }
  MaterialExpressionMultiply_35 {
   class=MaterialExpressionMultiply
   inputs [
    A {from=MaterialExpressionTextureSample_11, output=RGB}
    B {from=""}
   ]
  }
  MaterialExpressionMultiply_36 {
   class=MaterialExpressionMultiply
   inputs [
    A {from=MaterialExpressionTime_7}
    B {from=MaterialExpressionVectorParameter_10, output=RGB}
   ]
  }
  MaterialExpressionMultiply_37 {
   class=MaterialExpressionMultiply
   inputs [
    A {from=MaterialExpressionLinearInterpolate_36}
    B {from=MaterialExpressionAppendVector_12}
   ]
  }
  MaterialExpressionMultiply_38 {
   class=MaterialExpressionMultiply
   inputs [
    A {from=MaterialExpressionComponentMask_8}
    B {from=MaterialExpressionMax_5}
   ]
  }
  MaterialExpressionMultiply_39 {
   class=MaterialExpressionMultiply
   inputs [
    A {from=MaterialExpressionMultiply_40}
    B {from=MaterialExpressionDivide_8}
   ]
  }
  MaterialExpressionMultiply_4 {
   class=MaterialExpressionMultiply
   inputs [
    A {from=MaterialExpressionAdd_0}
    B {from=MaterialExpressionScalarParameter_5}
   ]
  }
  MaterialExpressionMultiply_40 {
   class=MaterialExpressionMultiply
   inputs [
    A {from=MaterialExpressionDivide_2}
    B {from=MaterialExpressionMaterialFunctionCall_47, output=Offset and Scaled UVS}
   ]
  }
  MaterialExpressionMultiply_41 {
   class=MaterialExpressionMultiply
   inputs [
    A {from=MaterialExpressionTextureSample_12, output=RGB}
    B {from=""}
   ]
  }
  MaterialExpressionMultiply_42 {
   class=MaterialExpressionMultiply
   inputs [
    A {from=MaterialExpressionLinearInterpolate_13}
    B {from=MaterialExpressionMaterialFunctionCall_37, output=Scale Z}
   ]
  }
  MaterialExpressionMultiply_43 {
   class=MaterialExpressionMultiply
   inputs [
    A {from=MaterialExpressionTime_8}
    B {from=MaterialExpressionConstant_11}
   ]
  }
  MaterialExpressionMultiply_44 {
   class=MaterialExpressionMultiply
   inputs [
    A {from=MaterialExpressionDesaturation_2}
    B {from=MaterialExpressionVectorParameter_23, output=RGB}
   ]
  }
  MaterialExpressionMultiply_46 {
   class=MaterialExpressionMultiply
   inputs [
    A {from=MaterialExpressionNamedRerouteUsage_67}
    B {from=MaterialExpressionSaturate_11}
   ]
   properties {constA=1}
  }
  MaterialExpressionMultiply_47 {
   class=MaterialExpressionMultiply
   inputs [
    A {from=MaterialExpressionVectorParameter_37, output=RGB}
    B {from=MaterialExpressionVectorParameter_37, output=RGB}
   ]
  }
  MaterialExpressionMultiply_48 {
   class=MaterialExpressionMultiply
   inputs [
    A {from=MaterialExpressionVectorParameter_38, output=RGB}
    B {from=MaterialExpressionVectorParameter_38, output=RGB}
   ]
  }
  MaterialExpressionMultiply_49 {
   class=MaterialExpressionMultiply
   inputs [
    A {from=MaterialExpressionVectorParameter_39, output=RGB}
    B {from=MaterialExpressionVectorParameter_39, output=RGB}
   ]
  }
  MaterialExpressionMultiply_5 {
   class=MaterialExpressionMultiply
   inputs [
    A {from=MaterialExpressionNamedRerouteUsage_8}
    B {from=MaterialExpressionScalarParameter_10}
   ]
  }
  MaterialExpressionMultiply_50 {
   class=MaterialExpressionMultiply
   inputs [
    A {from=MaterialExpressionVectorParameter_40, output=RGB}
    B {from=MaterialExpressionVectorParameter_40, output=RGB}
   ]
  }
  MaterialExpressionMultiply_51 {
   class=MaterialExpressionMultiply
   inputs [
    A {from=MaterialExpressionVectorParameter_41, output=RGB}
    B {from=MaterialExpressionVectorParameter_41, output=RGB}
   ]
  }
  MaterialExpressionMultiply_52 {
   class=MaterialExpressionMultiply
   inputs [
    A {from=MaterialExpressionVectorParameter_42, output=RGB}
    B {from=MaterialExpressionVectorParameter_42, output=RGB}
   ]
  }
  MaterialExpressionMultiply_53 {
   class=MaterialExpressionMultiply
   inputs [
    A {from=MaterialExpressionVectorParameter_43, output=RGB}
    B {from=MaterialExpressionVectorParameter_43, output=RGB}
   ]
  }
  MaterialExpressionMultiply_54 {
   class=MaterialExpressionMultiply
   inputs [
    A {from=MaterialExpressionSaturate_12}
    B {from=MaterialExpressionAdd_9}
   ]
  }
  MaterialExpressionMultiply_55 {
   class=MaterialExpressionMultiply
   inputs [
    A {from=MaterialExpressionSaturate_6}
    B {from=MaterialExpressionSaturate_13}
   ]
  }
  MaterialExpressionMultiply_56 {
   class=MaterialExpressionMultiply
   inputs [
    A {from=MaterialExpressionFunctionInput_11}
    B {from=MaterialExpressionScalarParameter_71}
   ]
  }
  MaterialExpressionMultiply_57 {
   class=MaterialExpressionMultiply
   inputs [
    A {from=MaterialExpressionFunctionInput_18}
    B {from=MaterialExpressionScalarParameter_70}
   ]
  }
  MaterialExpressionMultiply_58 {
   class=MaterialExpressionMultiply
   inputs [
    A {from=MaterialExpressionLinearInterpolate_79}
    B {from=MaterialExpressionSaturate_14}
   ]
  }
  MaterialExpressionMultiply_6 {
   class=MaterialExpressionMultiply
   inputs [
    A {from=MaterialExpressionReroute_0}
    B {from=MaterialExpressionScalarParameter_6}
   ]
  }
  MaterialExpressionMultiply_7 {
   class=MaterialExpressionMultiply
   inputs [
    A {from=MaterialExpressionTextureSample_9, output=R}
    B {from=MaterialExpressionTextureSample_9, output=R}
   ]
  }
  MaterialExpressionMultiply_8 {
   class=MaterialExpressionMultiply
   inputs [
    A {from=MaterialExpressionNamedRerouteUsage_17}
    B {from=MaterialExpressionScalarParameter_11}
   ]
  }
  MaterialExpressionMultiply_9 {
   class=MaterialExpressionMultiply
   inputs [
    A {from=MaterialExpressionTextureSample_7, output=B}
    B {from=MaterialExpressionTextureSample_7, output=B}
   ]
  }
  MaterialExpressionNamedRerouteDeclaration_0 {
   class=MaterialExpressionNamedRerouteDeclaration
   inputs [None {from=MaterialExpressionSetMaterialAttributes_2}]
   properties {name=Mat_Accent, nodeColor=<rgba 0 0.3906 1 1>}
  }
  MaterialExpressionNamedRerouteDeclaration_1 {
   class=MaterialExpressionNamedRerouteDeclaration
   inputs [None {from=MaterialExpressionFunctionInput_1}]
   properties {name=Msk_Curvature, nodeColor=<rgba 0 1 0.7422 1>}
  }
  MaterialExpressionNamedRerouteDeclaration_10 {
   class=MaterialExpressionNamedRerouteDeclaration
   inputs [None {from=MaterialExpressionGetMaterialAttributes_0, output=MaterialAttributes}]
   properties {name=Root_Material, nodeColor=<rgba 0.7812 1 0 1>}
  }
  MaterialExpressionNamedRerouteDeclaration_11 {
   class=MaterialExpressionNamedRerouteDeclaration
   inputs [None {from=MaterialExpressionTextureObjectParameter_1}]
   properties {name=Tex_Msk_Blend, nodeColor=<rgba 1 0 0.6094 1>}
  }
  MaterialExpressionNamedRerouteDeclaration_12 {
   class=MaterialExpressionNamedRerouteDeclaration
   inputs [None {from=MaterialExpressionMaterialFunctionCall_7, output=Result}]
   properties {name=Blend_Msk_A, nodeColor=<rgba 1 0.2109 0 1>}
  }
  MaterialExpressionNamedRerouteDeclaration_13 {
   class=MaterialExpressionNamedRerouteDeclaration
   inputs [None {from=MaterialExpressionTextureObjectParameter_4}]
   properties {name=Tex_Msk_1, nodeColor=<rgba 1 0 0.6094 1>}
  }
  MaterialExpressionNamedRerouteDeclaration_14 {
   class=MaterialExpressionNamedRerouteDeclaration
   inputs [None {from=MaterialExpressionTextureObjectParameter_2}]
   properties {name=Tex_BC, nodeColor=<rgba 1 0 0.6094 1>}
  }
  MaterialExpressionNamedRerouteDeclaration_15 {
   class=MaterialExpressionNamedRerouteDeclaration
   inputs [None {from=MaterialExpressionTextureObjectParameter_3}]
   properties {name=Tex_N, nodeColor=<rgba 1 0 0.6094 1>}
  }
  MaterialExpressionNamedRerouteDeclaration_16 {
   class=MaterialExpressionNamedRerouteDeclaration
   inputs [None {from=MaterialExpressionTextureObjectParameter_0}]
   properties {name=Tex_Stars_Msk, nodeColor=<rgba 1 0 0.6094 1>}
  }
  MaterialExpressionNamedRerouteDeclaration_17 {
   class=MaterialExpressionNamedRerouteDeclaration
   inputs [None {from=MaterialExpressionMultiply_4}]
   properties {name=Warp_Curv, nodeColor=<rgba 0.5469 1 0 1>}
  }
  MaterialExpressionNamedRerouteDeclaration_18 {
   class=MaterialExpressionNamedRerouteDeclaration
   inputs [None {from=MaterialExpressionSetMaterialAttributes_0}]
   properties {name=Mat_Galaxy, nodeColor=<rgba 0 0.3906 1 1>}
  }
  MaterialExpressionNamedRerouteDeclaration_19 {
   class=MaterialExpressionNamedRerouteDeclaration
   inputs [None {from=MaterialExpressionMaterialFunctionCall_4, output=UVs}]
   properties {name=Uvs, nodeColor=<rgba 0 1 0.8594 1>}
  }
  MaterialExpressionNamedRerouteDeclaration_2 {
   class=MaterialExpressionNamedRerouteDeclaration
   inputs [None {from=MaterialExpressionFunctionInput_9}]
   properties {name=Material_inline, nodeColor=<rgba 1 0.5625 0 1>}
  }
  MaterialExpressionNamedRerouteDeclaration_20 {
   class=MaterialExpressionNamedRerouteDeclaration
   inputs [None {from=MaterialExpressionTextureSample_1, output=RGB}]
   properties {name=Base_Base_Color, nodeColor=<rgba 1 0 0.4453 1>}
  }
  MaterialExpressionNamedRerouteDeclaration_21 {
   class=MaterialExpressionNamedRerouteDeclaration
   inputs [None {from=MaterialExpressionMaterialFunctionCall_4, output=Raw_Normal}]
   properties {name=Base_Normal, nodeColor=<rgba 1 0 0.4453 1>}
  }
  MaterialExpressionNamedRerouteDeclaration_22 {
   class=MaterialExpressionNamedRerouteDeclaration
   inputs [None {from=MaterialExpressionLinearInterpolate_12}]
   properties {name=UVs_Offset, nodeColor=<rgba 0 0.625 1 1>}
  }
  MaterialExpressionNamedRerouteDeclaration_23 {
   class=MaterialExpressionNamedRerouteDeclaration
   inputs [None {from=MaterialExpressionLinearInterpolate_9}]
   properties {name=UVs_Rotation, nodeColor=<rgba 0 0.7188 1 1>}
  }
  MaterialExpressionNamedRerouteDeclaration_24 {
   class=MaterialExpressionNamedRerouteDeclaration
   inputs [None {from=MaterialExpressionLinearInterpolate_16}]
   properties {name=UVs_Scale, nodeColor=<rgba 0 0.7188 1 1>}
  }
  MaterialExpressionNamedRerouteDeclaration_25 {
   class=MaterialExpressionNamedRerouteDeclaration
   inputs [None {from=MaterialExpressionLinearInterpolate_24}]
   properties {name=UVs_Rotation_TS, nodeColor=<rgba 0 0.7188 1 1>}
  }
  MaterialExpressionNamedRerouteDeclaration_26 {
   class=MaterialExpressionNamedRerouteDeclaration
   inputs [None {from=MaterialExpressionTextureSample_1, output=A}]
   properties {name=Base_Roughness, nodeColor=<rgba 1 0 0.4453 1>}
  }
  MaterialExpressionNamedRerouteDeclaration_27 {
   class=MaterialExpressionNamedRerouteDeclaration
   inputs [None {from=MaterialExpressionLinearInterpolate_0}]
   properties {name=Galaxy, nodeColor=<rgba 0 0.25 1 1>}
  }
  MaterialExpressionNamedRerouteDeclaration_28 {
   class=MaterialExpressionNamedRerouteDeclaration
   inputs [None {from=MaterialExpressionLinearInterpolate_30}]
   properties {name=Grad_Falloff, nodeColor=<rgba 0 0.7188 1 1>}
  }
  MaterialExpressionNamedRerouteDeclaration_29 {
   class=MaterialExpressionNamedRerouteDeclaration
   inputs [None {from=MaterialExpressionLinearInterpolate_1}]
   properties {name=Grad_offset, nodeColor=<rgba 0 0.7188 1 1>}
  }
  MaterialExpressionNamedRerouteDeclaration_3 {
   class=MaterialExpressionNamedRerouteDeclaration
   inputs [None {from=MaterialExpressionFunctionInput_11}]
   properties {name=Msk_Accent, nodeColor=<rgba 0 1 0.7656 1>}
  }
  MaterialExpressionNamedRerouteDeclaration_30 {
   class=MaterialExpressionNamedRerouteDeclaration
   inputs [None {from=MaterialExpressionMultiply_20}]
   properties {name=Glow_Colors, nodeColor=<rgba 0.2188 0 1 1>}
  }
  MaterialExpressionNamedRerouteDeclaration_31 {
   class=MaterialExpressionNamedRerouteDeclaration
   inputs [None {from=MaterialExpressionSetMaterialAttributes_4}]
   properties {name=Mat_Base, nodeColor=<rgba 0 0.3906 1 1>}
  }
  MaterialExpressionNamedRerouteDeclaration_32 {
   class=MaterialExpressionNamedRerouteDeclaration
   inputs [None {from=MaterialExpressionLinearInterpolate_19}]
   properties {name=UVs_Horn, nodeColor=<rgba 0 0.625 1 1>}
  }
  MaterialExpressionNamedRerouteDeclaration_33 {
   class=MaterialExpressionNamedRerouteDeclaration
   inputs [None {from=MaterialExpressionLinearInterpolate_52}]
   properties {name=UVs_Rotation_Horn, nodeColor=<rgba 0 0.7188 1 1>}
  }
  MaterialExpressionNamedRerouteDeclaration_34 {
   class=MaterialExpressionNamedRerouteDeclaration
   inputs [None {from=MaterialExpressionMultiply_12}]
   properties {name=Horn_Glow, nodeColor=<rgba 1 0 0.7734 1>}
  }
  MaterialExpressionNamedRerouteDeclaration_35 {
   class=MaterialExpressionNamedRerouteDeclaration
   inputs [None {from=MaterialExpressionMaterialFunctionCall_44, output=Result}]
   properties {name=Tri_Mask, nodeColor=<rgba 1 0.4688 0 1>}
  }
  MaterialExpressionNamedRerouteDeclaration_36 {
   class=MaterialExpressionNamedRerouteDeclaration
   inputs [None {from=MaterialExpressionTextureSample_5, output=R}]
   properties {name=Metalic, nodeColor=<rgba 0.4297 0 1 1>}
  }
  MaterialExpressionNamedRerouteDeclaration_37 {
   class=MaterialExpressionNamedRerouteDeclaration
   inputs [None {from=MaterialExpressionOneMinus_1}]
   properties {name=Horn_Msk, nodeColor=<rgba 1 0 0.7734 1>}
  }
  MaterialExpressionNamedRerouteDeclaration_38 {
   class=MaterialExpressionNamedRerouteDeclaration
   inputs [None {from=MaterialExpressionLinearInterpolate_61}]
   properties {name=UVs_Scale_Warp, nodeColor=<rgba 0 0.7188 1 1>}
  }
  MaterialExpressionNamedRerouteDeclaration_39 {
   class=MaterialExpressionNamedRerouteDeclaration
   inputs [None {from=MaterialExpressionMaterialFunctionCall_31, output=UVs}]
   properties {name=ScreenSpaceUVs, nodeColor=<rgba 0.5469 0 1 1>}
  }
  MaterialExpressionNamedRerouteDeclaration_4 {
   class=MaterialExpressionNamedRerouteDeclaration
   inputs [None {from=MaterialExpressionFunctionInput_3}]
   properties {name=IsState, nodeColor=<rgba 0.1497 1 0.1633 1>}
  }
  MaterialExpressionNamedRerouteDeclaration_40 {
   class=MaterialExpressionNamedRerouteDeclaration
   inputs [None {from=MaterialExpressionLinearInterpolate_43}]
   properties {name=Texture_Scale, nodeColor=<rgba 0 0.7188 1 1>}
  }
  MaterialExpressionNamedRerouteDeclaration_41 {
   class=MaterialExpressionNamedRerouteDeclaration
   inputs [None {from=MaterialExpressionFunctionInput_30}]
   properties {name=SafeZone VP, nodeColor=<rgba 0.5 1 0 1>}
  }
  MaterialExpressionNamedRerouteDeclaration_42 {
   class=MaterialExpressionNamedRerouteDeclaration
   inputs [None {from=MaterialExpressionFunctionInput_7}]
   properties {name=SafeZone Character-Switch, nodeColor=<rgba 1 0.9375 0 1>}
  }
  MaterialExpressionNamedRerouteDeclaration_43 {
   class=MaterialExpressionNamedRerouteDeclaration
   inputs [None {from=MaterialExpressionLinearInterpolate_67}]
   properties {name=Fov_State, nodeColor=<rgba 0 0.7188 1 1>}
  }
  MaterialExpressionNamedRerouteDeclaration_44 {
   class=MaterialExpressionNamedRerouteDeclaration
   inputs [None {from=MaterialExpressionLinearInterpolate_72}]
   properties {name=opacity, nodeColor=<rgba 0 0.7188 1 1>}
  }
  MaterialExpressionNamedRerouteDeclaration_45 {
   class=MaterialExpressionNamedRerouteDeclaration
   inputs [None {from=MaterialExpressionPower_0}]
   properties {name=Edgeing_Msk, nodeColor=<rgba 0.1719 1 0 1>}
  }
  MaterialExpressionNamedRerouteDeclaration_46 {
   class=MaterialExpressionNamedRerouteDeclaration
   inputs [None {from=MaterialExpressionFunctionInput_4}]
   properties {name=RootRoot_Material, nodeColor=<rgba 0.7812 1 0 1>}
  }
  MaterialExpressionNamedRerouteDeclaration_47 {
   class=MaterialExpressionNamedRerouteDeclaration
   inputs [None {from=MaterialExpressionTextureSample_1, output=RGBA}]
   properties {name=Msk_Side Projection, nodeColor=<rgba 1 0 0.4453 1>}
  }
  MaterialExpressionNamedRerouteDeclaration_48 {
   class=MaterialExpressionNamedRerouteDeclaration
   inputs [None {from=MaterialExpressionLinearInterpolate_75}]
   properties {name=Msk_offset_B, nodeColor=<rgba 0 0.7188 1 1>}
  }
  MaterialExpressionNamedRerouteDeclaration_49 {
   class=MaterialExpressionNamedRerouteDeclaration
   inputs [None {from=MaterialExpressionGetMaterialAttributes_0, output=EmissiveColor}]
   properties {name=Root_Emissive, nodeColor=<rgba 0.4062 0 1 1>}
  }
  MaterialExpressionNamedRerouteDeclaration_5 {
   class=MaterialExpressionNamedRerouteDeclaration
   inputs [None {from=MaterialExpressionFunctionInput_10}]
   properties {name=IsState_B, nodeColor=<rgba 0.1497 1 0.1633 1>}
  }
  MaterialExpressionNamedRerouteDeclaration_6 {
   class=MaterialExpressionNamedRerouteDeclaration
   inputs [None {from=MaterialExpressionFunctionInput_2}]
   properties {name=Caclcated_Bounds, nodeColor=<rgba 0 0.9297 1 1>}
  }
  MaterialExpressionNamedRerouteDeclaration_7 {
   class=MaterialExpressionNamedRerouteDeclaration
   inputs [None {from=MaterialExpressionTextureObjectParameter_7}]
   properties {name=Tex_Galx, nodeColor=<rgba 1 0 0.6094 1>}
  }
  MaterialExpressionNamedRerouteDeclaration_8 {
   class=MaterialExpressionNamedRerouteDeclaration
   inputs [None {from=MaterialExpressionFunctionInput_0}]
   properties {name=PreSkin_Normal, nodeColor=<rgba 0.8281 1 0 1>}
  }
  MaterialExpressionNamedRerouteDeclaration_9 {
   class=MaterialExpressionNamedRerouteDeclaration
   inputs [None {from=MaterialExpressionGetMaterialAttributes_0, output=Normal}]
   properties {name=Root_Normal, nodeColor=<rgba 0.4062 0 1 1>}
  }
  MaterialExpressionNamedRerouteUsage_0 {class=MaterialExpressionNamedRerouteUsage}
  MaterialExpressionNamedRerouteUsage_1 {class=MaterialExpressionNamedRerouteUsage}
  MaterialExpressionNamedRerouteUsage_10 {class=MaterialExpressionNamedRerouteUsage}
  MaterialExpressionNamedRerouteUsage_100 {class=MaterialExpressionNamedRerouteUsage}
  MaterialExpressionNamedRerouteUsage_101 {class=MaterialExpressionNamedRerouteUsage}
  MaterialExpressionNamedRerouteUsage_102 {class=MaterialExpressionNamedRerouteUsage}
  MaterialExpressionNamedRerouteUsage_103 {class=MaterialExpressionNamedRerouteUsage}
  MaterialExpressionNamedRerouteUsage_104 {class=MaterialExpressionNamedRerouteUsage}
  MaterialExpressionNamedRerouteUsage_105 {class=MaterialExpressionNamedRerouteUsage}
  MaterialExpressionNamedRerouteUsage_106 {class=MaterialExpressionNamedRerouteUsage}
  MaterialExpressionNamedRerouteUsage_11 {class=MaterialExpressionNamedRerouteUsage}
  MaterialExpressionNamedRerouteUsage_12 {class=MaterialExpressionNamedRerouteUsage}
  MaterialExpressionNamedRerouteUsage_13 {class=MaterialExpressionNamedRerouteUsage}
  MaterialExpressionNamedRerouteUsage_14 {class=MaterialExpressionNamedRerouteUsage}
  MaterialExpressionNamedRerouteUsage_15 {class=MaterialExpressionNamedRerouteUsage}
  MaterialExpressionNamedRerouteUsage_16 {class=MaterialExpressionNamedRerouteUsage}
  MaterialExpressionNamedRerouteUsage_17 {class=MaterialExpressionNamedRerouteUsage}
  MaterialExpressionNamedRerouteUsage_18 {class=MaterialExpressionNamedRerouteUsage}
  MaterialExpressionNamedRerouteUsage_19 {class=MaterialExpressionNamedRerouteUsage}
  MaterialExpressionNamedRerouteUsage_2 {class=MaterialExpressionNamedRerouteUsage}
  MaterialExpressionNamedRerouteUsage_20 {class=MaterialExpressionNamedRerouteUsage}
  MaterialExpressionNamedRerouteUsage_21 {class=MaterialExpressionNamedRerouteUsage}
  MaterialExpressionNamedRerouteUsage_22 {class=MaterialExpressionNamedRerouteUsage}
  MaterialExpressionNamedRerouteUsage_23 {class=MaterialExpressionNamedRerouteUsage}
  MaterialExpressionNamedRerouteUsage_24 {class=MaterialExpressionNamedRerouteUsage}
  MaterialExpressionNamedRerouteUsage_25 {class=MaterialExpressionNamedRerouteUsage}
  MaterialExpressionNamedRerouteUsage_26 {class=MaterialExpressionNamedRerouteUsage}
  MaterialExpressionNamedRerouteUsage_27 {class=MaterialExpressionNamedRerouteUsage}
  MaterialExpressionNamedRerouteUsage_28 {class=MaterialExpressionNamedRerouteUsage}
  MaterialExpressionNamedRerouteUsage_29 {class=MaterialExpressionNamedRerouteUsage}
  MaterialExpressionNamedRerouteUsage_3 {class=MaterialExpressionNamedRerouteUsage}
  MaterialExpressionNamedRerouteUsage_30 {class=MaterialExpressionNamedRerouteUsage}
  MaterialExpressionNamedRerouteUsage_31 {class=MaterialExpressionNamedRerouteUsage}
  MaterialExpressionNamedRerouteUsage_32 {class=MaterialExpressionNamedRerouteUsage}
  MaterialExpressionNamedRerouteUsage_33 {class=MaterialExpressionNamedRerouteUsage}
  MaterialExpressionNamedRerouteUsage_34 {class=MaterialExpressionNamedRerouteUsage}
  MaterialExpressionNamedRerouteUsage_35 {class=MaterialExpressionNamedRerouteUsage}
  MaterialExpressionNamedRerouteUsage_36 {class=MaterialExpressionNamedRerouteUsage}
  MaterialExpressionNamedRerouteUsage_37 {class=MaterialExpressionNamedRerouteUsage}
  MaterialExpressionNamedRerouteUsage_38 {class=MaterialExpressionNamedRerouteUsage}
  MaterialExpressionNamedRerouteUsage_39 {class=MaterialExpressionNamedRerouteUsage}
  MaterialExpressionNamedRerouteUsage_4 {class=MaterialExpressionNamedRerouteUsage}
  MaterialExpressionNamedRerouteUsage_40 {class=MaterialExpressionNamedRerouteUsage}
  MaterialExpressionNamedRerouteUsage_41 {class=MaterialExpressionNamedRerouteUsage}
  MaterialExpressionNamedRerouteUsage_42 {class=MaterialExpressionNamedRerouteUsage}
  MaterialExpressionNamedRerouteUsage_43 {class=MaterialExpressionNamedRerouteUsage}
  MaterialExpressionNamedRerouteUsage_44 {class=MaterialExpressionNamedRerouteUsage}
  MaterialExpressionNamedRerouteUsage_45 {class=MaterialExpressionNamedRerouteUsage}
  MaterialExpressionNamedRerouteUsage_46 {class=MaterialExpressionNamedRerouteUsage}
  MaterialExpressionNamedRerouteUsage_47 {class=MaterialExpressionNamedRerouteUsage}
  MaterialExpressionNamedRerouteUsage_48 {class=MaterialExpressionNamedRerouteUsage}
  MaterialExpressionNamedRerouteUsage_49 {class=MaterialExpressionNamedRerouteUsage}
  MaterialExpressionNamedRerouteUsage_5 {class=MaterialExpressionNamedRerouteUsage}
  MaterialExpressionNamedRerouteUsage_50 {class=MaterialExpressionNamedRerouteUsage}
  MaterialExpressionNamedRerouteUsage_51 {class=MaterialExpressionNamedRerouteUsage}
  MaterialExpressionNamedRerouteUsage_52 {class=MaterialExpressionNamedRerouteUsage}
  MaterialExpressionNamedRerouteUsage_53 {class=MaterialExpressionNamedRerouteUsage}
  MaterialExpressionNamedRerouteUsage_54 {class=MaterialExpressionNamedRerouteUsage}
  MaterialExpressionNamedRerouteUsage_55 {class=MaterialExpressionNamedRerouteUsage}
  MaterialExpressionNamedRerouteUsage_56 {class=MaterialExpressionNamedRerouteUsage}
  MaterialExpressionNamedRerouteUsage_57 {class=MaterialExpressionNamedRerouteUsage}
  MaterialExpressionNamedRerouteUsage_58 {class=MaterialExpressionNamedRerouteUsage}
  MaterialExpressionNamedRerouteUsage_59 {class=MaterialExpressionNamedRerouteUsage}
  MaterialExpressionNamedRerouteUsage_6 {class=MaterialExpressionNamedRerouteUsage}
  MaterialExpressionNamedRerouteUsage_60 {class=MaterialExpressionNamedRerouteUsage}
  MaterialExpressionNamedRerouteUsage_61 {class=MaterialExpressionNamedRerouteUsage}
  MaterialExpressionNamedRerouteUsage_62 {class=MaterialExpressionNamedRerouteUsage}
  MaterialExpressionNamedRerouteUsage_63 {class=MaterialExpressionNamedRerouteUsage}
  MaterialExpressionNamedRerouteUsage_64 {class=MaterialExpressionNamedRerouteUsage}
  MaterialExpressionNamedRerouteUsage_65 {class=MaterialExpressionNamedRerouteUsage}
  MaterialExpressionNamedRerouteUsage_66 {class=MaterialExpressionNamedRerouteUsage}
  MaterialExpressionNamedRerouteUsage_67 {class=MaterialExpressionNamedRerouteUsage}
  MaterialExpressionNamedRerouteUsage_68 {class=MaterialExpressionNamedRerouteUsage}
  MaterialExpressionNamedRerouteUsage_69 {class=MaterialExpressionNamedRerouteUsage}
  MaterialExpressionNamedRerouteUsage_7 {class=MaterialExpressionNamedRerouteUsage}
  MaterialExpressionNamedRerouteUsage_70 {class=MaterialExpressionNamedRerouteUsage}
  MaterialExpressionNamedRerouteUsage_71 {class=MaterialExpressionNamedRerouteUsage}
  MaterialExpressionNamedRerouteUsage_72 {class=MaterialExpressionNamedRerouteUsage}
  MaterialExpressionNamedRerouteUsage_73 {class=MaterialExpressionNamedRerouteUsage}
  MaterialExpressionNamedRerouteUsage_74 {class=MaterialExpressionNamedRerouteUsage}
  MaterialExpressionNamedRerouteUsage_75 {class=MaterialExpressionNamedRerouteUsage}
  MaterialExpressionNamedRerouteUsage_76 {class=MaterialExpressionNamedRerouteUsage}
  MaterialExpressionNamedRerouteUsage_77 {class=MaterialExpressionNamedRerouteUsage}
  MaterialExpressionNamedRerouteUsage_78 {class=MaterialExpressionNamedRerouteUsage}
  MaterialExpressionNamedRerouteUsage_79 {class=MaterialExpressionNamedRerouteUsage}
  MaterialExpressionNamedRerouteUsage_8 {class=MaterialExpressionNamedRerouteUsage}
  MaterialExpressionNamedRerouteUsage_80 {class=MaterialExpressionNamedRerouteUsage}
  MaterialExpressionNamedRerouteUsage_81 {class=MaterialExpressionNamedRerouteUsage}
  MaterialExpressionNamedRerouteUsage_82 {class=MaterialExpressionNamedRerouteUsage}
  MaterialExpressionNamedRerouteUsage_83 {class=MaterialExpressionNamedRerouteUsage}
  MaterialExpressionNamedRerouteUsage_84 {class=MaterialExpressionNamedRerouteUsage}
  MaterialExpressionNamedRerouteUsage_85 {class=MaterialExpressionNamedRerouteUsage}
  MaterialExpressionNamedRerouteUsage_86 {class=MaterialExpressionNamedRerouteUsage}
  MaterialExpressionNamedRerouteUsage_87 {class=MaterialExpressionNamedRerouteUsage}
  MaterialExpressionNamedRerouteUsage_88 {class=MaterialExpressionNamedRerouteUsage}
  MaterialExpressionNamedRerouteUsage_89 {class=MaterialExpressionNamedRerouteUsage}
  MaterialExpressionNamedRerouteUsage_9 {class=MaterialExpressionNamedRerouteUsage}
  MaterialExpressionNamedRerouteUsage_90 {class=MaterialExpressionNamedRerouteUsage}
  MaterialExpressionNamedRerouteUsage_91 {class=MaterialExpressionNamedRerouteUsage}
  MaterialExpressionNamedRerouteUsage_92 {class=MaterialExpressionNamedRerouteUsage}
  MaterialExpressionNamedRerouteUsage_93 {class=MaterialExpressionNamedRerouteUsage}
  MaterialExpressionNamedRerouteUsage_94 {class=MaterialExpressionNamedRerouteUsage}
  MaterialExpressionNamedRerouteUsage_95 {class=MaterialExpressionNamedRerouteUsage}
  MaterialExpressionNamedRerouteUsage_96 {class=MaterialExpressionNamedRerouteUsage}
  MaterialExpressionNamedRerouteUsage_97 {class=MaterialExpressionNamedRerouteUsage}
  MaterialExpressionNamedRerouteUsage_98 {class=MaterialExpressionNamedRerouteUsage}
  MaterialExpressionNamedRerouteUsage_99 {class=MaterialExpressionNamedRerouteUsage}
  MaterialExpressionNormalize_2 {class=MaterialExpressionNormalize, inputs [VectorInput {from=MaterialExpressionAdd_21, output=""}]}
  MaterialExpressionNormalize_3 {class=MaterialExpressionNormalize, inputs [VectorInput {from=MaterialExpressionMultiply_25, output=""}]}
  MaterialExpressionOneMinus_0 {class=MaterialExpressionOneMinus, inputs [None {from=MaterialExpressionMaterialFunctionCall_8, output=Result}]}
  MaterialExpressionOneMinus_1 {class=MaterialExpressionOneMinus, inputs [None {from=MaterialExpressionMultiply_13}]}
  MaterialExpressionOneMinus_2 {class=MaterialExpressionOneMinus, inputs [None {from=MaterialExpressionPower_4}]}
  MaterialExpressionOneMinus_3 {class=MaterialExpressionOneMinus, inputs [None {from=MaterialExpressionNamedRerouteUsage_95}]}
  MaterialExpressionOneMinus_4 {class=MaterialExpressionOneMinus, inputs [None {from=MaterialExpressionNamedRerouteUsage_96}]}
  MaterialExpressionPanner_0 {
   class=MaterialExpressionPanner
   inputs [
    Coordinate {from=MaterialExpressionNamedRerouteUsage_40}
    Time {from=""}
    Speed {from=MaterialExpressionComponentMask_3}
   ]
   properties {speedY=0.02}
  }
  MaterialExpressionPanner_1 {
   class=MaterialExpressionPanner
   inputs [
    Coordinate {from=MaterialExpressionMultiply_5}
    Time {from=""}
    Speed {from=MaterialExpressionComponentMask_4}
   ]
   properties {speedY=0.02}
  }
  MaterialExpressionPanner_2 {
   class=MaterialExpressionPanner
   inputs [
    Coordinate {from=MaterialExpressionNamedRerouteUsage_29}
    Time {from=""}
    Speed {from=MaterialExpressionComponentMask_10}
   ]
   properties {speedX=0.133, speedY=0.2}
  }
  MaterialExpressionPanner_3 {
   class=MaterialExpressionPanner
   inputs [
    Coordinate {from=MaterialExpressionTextureCoordinate_7}
    Time {from=""}
    Speed {from=""}
   ]
   properties {speedX=0.06, speedY=0.1}
  }
  MaterialExpressionPanner_4 {
   class=MaterialExpressionPanner
   inputs [
    Coordinate {from=MaterialExpressionMultiply_8}
    Time {from=""}
    Speed {from=MaterialExpressionComponentMask_11}
   ]
   properties {speedX=0.133, speedY=0.2}
  }
  MaterialExpressionPower_0 {
   class=MaterialExpressionPower
   inputs [
    Base {from=MaterialExpressionSaturate_3}
    Exp {from=""}
   ]
  }
  MaterialExpressionPower_1 {
   class=MaterialExpressionPower
   inputs [
    Base {from=MaterialExpressionTextureSample_2, output=G}
    Exp {from=""}
   ]
  }
  MaterialExpressionPower_2 {
   class=MaterialExpressionPower
   inputs [
    Base {from=MaterialExpressionMaterialFunctionCall_28, output=Result}
    Exp {from=MaterialExpressionScalarParameter_27}
   ]
   properties {constExponent=5}
  }
  MaterialExpressionPower_4 {
   class=MaterialExpressionPower
   inputs [
    Base {from=MaterialExpressionMaterialFunctionCall_8, output=Result}
    Exp {from=""}
   ]
   properties {constExponent=1.5}
  }
  MaterialExpressionPreSkinnedLocalBounds_1 {class=MaterialExpressionPreSkinnedLocalBounds, inputs [], properties {}}
  MaterialExpressionPreSkinnedLocalBounds_3 {class=MaterialExpressionPreSkinnedLocalBounds, inputs [], properties {}}
  MaterialExpressionPreSkinnedNormal_1 {class=MaterialExpressionPreSkinnedNormal, inputs [], properties {}}
  MaterialExpressionPreSkinnedPosition_0 {class=MaterialExpressionPreSkinnedPosition, inputs [], properties {}}
  MaterialExpressionReroute_0 {class=MaterialExpressionReroute, inputs [None {from=MaterialExpressionLinearInterpolate_18}]}
  MaterialExpressionReroute_1 {class=MaterialExpressionReroute, inputs [None {from=MaterialExpressionNamedRerouteUsage_43}]}
  MaterialExpressionReroute_2 {class=MaterialExpressionReroute, inputs [None {from=MaterialExpressionNamedRerouteUsage_44}]}
  MaterialExpressionReroute_3 {class=MaterialExpressionReroute, inputs [None {from=MaterialExpressionNamedRerouteUsage_43}]}
  MaterialExpressionReroute_31 {class=MaterialExpressionReroute, inputs [None {from=MaterialExpressionMultiply_9}]}
  MaterialExpressionReroute_32 {class=MaterialExpressionReroute, inputs [None {from=MaterialExpressionReroute_37}]}
  MaterialExpressionReroute_33 {class=MaterialExpressionReroute, inputs [None {from=MaterialExpressionMultiply_54}]}
  MaterialExpressionReroute_35 {class=MaterialExpressionReroute, inputs [None {from=MaterialExpressionMultiply_37}]}
  MaterialExpressionReroute_36 {class=MaterialExpressionReroute, inputs [None {from=MaterialExpressionNamedRerouteUsage_9}]}
  MaterialExpressionReroute_37 {class=MaterialExpressionReroute, inputs [None {from=MaterialExpressionDesaturation_2}]}
  MaterialExpressionReroute_4 {class=MaterialExpressionReroute, inputs [None {from=MaterialExpressionReroute_5}]}
  MaterialExpressionReroute_5 {class=MaterialExpressionReroute, inputs [None {from=MaterialExpressionInverseLinearInterpolate_1}]}
  MaterialExpressionReroute_6 {class=MaterialExpressionReroute, inputs [None {from=MaterialExpressionReroute_7}]}
  MaterialExpressionReroute_7 {class=MaterialExpressionReroute, inputs [None {from=MaterialExpressionScalarParameter_86}]}
  MaterialExpressionSaturate_0 {class=MaterialExpressionSaturate, inputs [None {from=MaterialExpressionPower_2}]}
  MaterialExpressionSaturate_1 {class=MaterialExpressionSaturate, inputs [None {from=MaterialExpressionAdd_24}]}
  MaterialExpressionSaturate_10 {class=MaterialExpressionSaturate, inputs [None {from=MaterialExpressionAdd_17}]}
  MaterialExpressionSaturate_11 {class=MaterialExpressionSaturate, inputs [None {from=MaterialExpressionAdd_8}]}
  MaterialExpressionSaturate_12 {class=MaterialExpressionSaturate, inputs [None {from=MaterialExpressionAdd_13}]}
  MaterialExpressionSaturate_13 {class=MaterialExpressionSaturate, inputs [None {from=MaterialExpressionAdd_20}]}
  MaterialExpressionSaturate_14 {class=MaterialExpressionSaturate, inputs [None {from=MaterialExpressionStaticSwitchParameter_2}]}
  MaterialExpressionSaturate_15 {class=MaterialExpressionSaturate, inputs [None {from=MaterialExpressionMax_1}]}
  MaterialExpressionSaturate_2 {class=MaterialExpressionSaturate, inputs [None {from=MaterialExpressionMultiply_22}]}
  MaterialExpressionSaturate_3 {class=MaterialExpressionSaturate, inputs [None {from=MaterialExpressionOneMinus_0}]}
  MaterialExpressionSaturate_4 {class=MaterialExpressionSaturate, inputs [None {from=MaterialExpressionAdd_7}]}
  MaterialExpressionSaturate_5 {class=MaterialExpressionSaturate, inputs [None {from=MaterialExpressionMultiply_29}]}
  MaterialExpressionSaturate_6 {class=MaterialExpressionSaturate, inputs [None {from=MaterialExpressionAdd_11}]}
  MaterialExpressionSaturate_7 {class=MaterialExpressionSaturate, inputs [None {from=MaterialExpressionAdd_12}]}
  MaterialExpressionSaturate_8 {class=MaterialExpressionSaturate, inputs [None {from=MaterialExpressionMaterialFunctionCall_20, output=UVs}]}
  MaterialExpressionSaturate_9 {class=MaterialExpressionSaturate, inputs [None {from=MaterialExpressionMultiply_15}]}
  MaterialExpressionScalarParameter_0 {class=MaterialExpressionScalarParameter, properties {defaultValue=1, group=Wrap_NebulaV2_Blend, parameterName=Neb_Fill_Alternate_Axies}}
  MaterialExpressionScalarParameter_1 {
   class=MaterialExpressionScalarParameter
   properties {defaultValue=-0.5, group=Wrap_NebulaV2_Grad, parameterName=Neb_Grad_Offset_Backpack, sortPriority=150}
  }
  MaterialExpressionScalarParameter_10 {class=MaterialExpressionScalarParameter, properties {defaultValue=8, group=Wrap_NebulaV2_Galaxy, parameterName=Neb_Small Star Tiling}}
  MaterialExpressionScalarParameter_11 {class=MaterialExpressionScalarParameter, properties {defaultValue=3, group=Wrap_NebulaV2_Galaxy, parameterName=Neb_Bright Star Tiling}}
  MaterialExpressionScalarParameter_12 {class=MaterialExpressionScalarParameter, properties {defaultValue=1.25, group=Wrap_NebulaV2_Blend, parameterName=Neb_Blend_Warp_A_Amount}}
  MaterialExpressionScalarParameter_13 {
   class=MaterialExpressionScalarParameter
   properties {defaultValue=1, group=Wrap_NebulaV2_Uvs, parameterName=Neb_Scale_Vehicle, sortPriority=150}
  }
  MaterialExpressionScalarParameter_14 {
   class=MaterialExpressionScalarParameter
   properties {defaultValue=1, group=Wrap_NebulaV2_Uvs, parameterName=Neb_Scale_Vehicle_Side, sortPriority=150}
  }
  MaterialExpressionScalarParameter_15 {class=MaterialExpressionScalarParameter, properties {group=Wrap_NebulaV2_Uvs, parameterName=Neb_Rotation_Vehicle, sortPriority=150}}
  MaterialExpressionScalarParameter_16 {class=MaterialExpressionScalarParameter, properties {group=Wrap_NebulaV2_Uvs, parameterName=Neb_Rotation_Vehicle_Side, sortPriority=150}}
  MaterialExpressionScalarParameter_17 {
   class=MaterialExpressionScalarParameter
   properties {group=Wrap_NebulaV2_Uvs, parameterName="Neb_Rotation_Weapon\\Pickaxe", sortPriority=150}
  }
  MaterialExpressionScalarParameter_18 {
   class=MaterialExpressionScalarParameter
   properties {group=Wrap_NebulaV2_Uvs, parameterName="Neb_Rotation_Mech\\Character\\Backpack", sortPriority=150}
  }
  MaterialExpressionScalarParameter_19 {class=MaterialExpressionScalarParameter, properties {defaultValue=10, group=Wrap_NebulaV2_Blend, parameterName=Neb_Blend_Contrast}}
  MaterialExpressionScalarParameter_2 {class=MaterialExpressionScalarParameter, properties {defaultValue=0.58, group=Wrap_NebulaV2_Accent, parameterName=Neb_Accent_Roughness}}
  MaterialExpressionScalarParameter_20 {class=MaterialExpressionScalarParameter, properties {defaultValue=0.25, group=Wrap_NebulaV2_Blend, parameterName=Neb_Blend_Glow_Contrast}}
  MaterialExpressionScalarParameter_21 {
   class=MaterialExpressionScalarParameter
   properties {defaultValue=1, group=Wrap_NebulaV2_Uvs, parameterName=Neb_Scale_Weapon, sortPriority=150}
  }
  MaterialExpressionScalarParameter_22 {
   class=MaterialExpressionScalarParameter
   properties {defaultValue=1, group=Wrap_NebulaV2_Uvs, parameterName=Neb_Scale_Mech, sortPriority=150}
  }
  MaterialExpressionScalarParameter_23 {
   class=MaterialExpressionScalarParameter
   properties {defaultValue=1, group=Wrap_NebulaV2_Uvs, parameterName=Neb_Scale_Backpack, sortPriority=150}
  }
  MaterialExpressionScalarParameter_24 {
   class=MaterialExpressionScalarParameter
   properties {defaultValue=1, group=Wrap_NebulaV2_Uvs, parameterName=Neb_Scale_Pickaxe, sortPriority=150}
  }
  MaterialExpressionScalarParameter_25 {
   class=MaterialExpressionScalarParameter
   properties {defaultValue=1, group=Wrap_NebulaV2_Uvs, parameterName=Neb_Scale_Character, sortPriority=150}
  }
  MaterialExpressionScalarParameter_26 {class=MaterialExpressionScalarParameter, properties {defaultValue=1, group=Wrap_NebulaV2_Blend, parameterName=Neb_Blend_Glow_Ab_Contrast}}
  MaterialExpressionScalarParameter_27 {class=MaterialExpressionScalarParameter, properties {defaultValue=0.5, group=Wrap_NebulaV2_Blend, parameterName=Neb_Blend_Glow_Ab_Power}}
  MaterialExpressionScalarParameter_28 {class=MaterialExpressionScalarParameter, properties {defaultValue=10, group=Wrap_NebulaV2_Blend, parameterName=Neb_TriBlend_Contrast}}
  MaterialExpressionScalarParameter_29 {class=MaterialExpressionScalarParameter, properties {defaultValue=0.85, group=Wrap_NebulaV2, parameterName=Neb_Rougness_Edge}}
  MaterialExpressionScalarParameter_3 {class=MaterialExpressionScalarParameter, properties {defaultValue=0.5, group=Wrap_NebulaV2_Accent, parameterName=Neb_Accent_Spec}}
  MaterialExpressionScalarParameter_30 {class=MaterialExpressionScalarParameter, properties {group=Wrap_NebulaV2, parameterName=Neb_Spec}}
  MaterialExpressionScalarParameter_31 {class=MaterialExpressionScalarParameter, properties {defaultValue=0.5, group=Wrap_NebulaV2, parameterName=Neb_Spec_Edge}}
  MaterialExpressionScalarParameter_32 {
   class=MaterialExpressionScalarParameter
   properties {defaultValue=0.5, group=Wrap_NebulaV2_Grad, parameterName=Neb_Grad_Falloff_Vehicle, sortPriority=150}
  }
  MaterialExpressionScalarParameter_33 {
   class=MaterialExpressionScalarParameter
   properties {defaultValue=0.5, group=Wrap_NebulaV2_Grad, parameterName=Neb_Grad_Falloff_Vehicle_Side, sortPriority=150}
  }
  MaterialExpressionScalarParameter_34 {
   class=MaterialExpressionScalarParameter
   properties {defaultValue=0.5, group=Wrap_NebulaV2_Grad, parameterName=Neb_Grad_Falloff_Weapon, sortPriority=150}
  }
  MaterialExpressionScalarParameter_35 {
   class=MaterialExpressionScalarParameter
   properties {defaultValue=0.5, group=Wrap_NebulaV2_Grad, parameterName=Neb_Grad_Falloff_Mech, sortPriority=150}
  }
  MaterialExpressionScalarParameter_36 {
   class=MaterialExpressionScalarParameter
   properties {defaultValue=0.5, group=Wrap_NebulaV2_Grad, parameterName=Neb_Grad_Falloff_Backpack, sortPriority=150}
  }
  MaterialExpressionScalarParameter_37 {
   class=MaterialExpressionScalarParameter
   properties {defaultValue=0.5, group=Wrap_NebulaV2_Grad, parameterName=Neb_Grad_Falloff_Pickaxe, sortPriority=150}
  }
  MaterialExpressionScalarParameter_38 {
   class=MaterialExpressionScalarParameter
   properties {defaultValue=0.5, group=Wrap_NebulaV2_Grad, parameterName=Neb_Grad_Falloff_Character, sortPriority=150}
  }
  MaterialExpressionScalarParameter_39 {
   class=MaterialExpressionScalarParameter
   properties {defaultValue=-0.5, group=Wrap_NebulaV2_Grad, parameterName=Neb_Grad_Offset_Vehicle, sortPriority=150}
  }
  MaterialExpressionScalarParameter_4 {class=MaterialExpressionScalarParameter, properties {group=Wrap_NebulaV2_Accent, parameterName=Neb_Accent_Metalic}}
  MaterialExpressionScalarParameter_40 {
   class=MaterialExpressionScalarParameter
   properties {defaultValue=-0.5, group=Wrap_NebulaV2_Grad, parameterName=Neb_Grad_Offset_Vehicle_Side, sortPriority=150}
  }
  MaterialExpressionScalarParameter_41 {
   class=MaterialExpressionScalarParameter
   properties {defaultValue=-0.5, group=Wrap_NebulaV2_Grad, parameterName=Neb_Grad_Offset_Weapon, sortPriority=150}
  }
  MaterialExpressionScalarParameter_42 {
   class=MaterialExpressionScalarParameter
   properties {defaultValue=-0.5, group=Wrap_NebulaV2_Grad, parameterName=Neb_Grad_Offset_Mech, sortPriority=150}
  }
  MaterialExpressionScalarParameter_43 {class=MaterialExpressionScalarParameter, properties {defaultValue=1, group=Wrap_NebulaV2_Galaxy, parameterName=Neb_Galaxy_Tiling}}
  MaterialExpressionScalarParameter_44 {
   class=MaterialExpressionScalarParameter
   properties {defaultValue=128, group=Wrap_NebulaV2_Galaxy, parameterName=Neb__TextureResolution, sliderMax=2048}
  }
  MaterialExpressionScalarParameter_45 {
   class=MaterialExpressionScalarParameter
   properties {defaultValue=5, group=Wrap_NebulaV2_Galaxy, parameterName=Neb_Galaxy_Fres_EX, sortPriority=-10}
  }
  MaterialExpressionScalarParameter_46 {class=MaterialExpressionScalarParameter, properties {defaultValue=8, group=Wrap_NebulaV2_Galaxy, parameterName=Neb_Small Star Tiling}}
  MaterialExpressionScalarParameter_47 {class=MaterialExpressionScalarParameter, properties {defaultValue=3, group=Wrap_NebulaV2_Galaxy, parameterName=Neb_Bright Star Tiling}}
  MaterialExpressionScalarParameter_48 {
   class=MaterialExpressionScalarParameter
   properties {defaultValue=-0.0125, group=Wrap_NebulaV2_Galaxy, parameterName=Neb_Small Star Rotation speed}
  }
  MaterialExpressionScalarParameter_49 {class=MaterialExpressionScalarParameter, properties {defaultValue=0.5, group=Wrap_NebulaV2_Galaxy, parameterName=EGal_Star Saturation}}
  MaterialExpressionScalarParameter_5 {class=MaterialExpressionScalarParameter, properties {group=Wrap_NebulaV2_Galaxy, parameterName=Neb_Warp_Curvature_Strength}}
  MaterialExpressionScalarParameter_50 {class=MaterialExpressionScalarParameter, properties {defaultValue=0.5, group=Wrap_NebulaV2_Galaxy, parameterName=Neb_GalaxyTiling}}
  MaterialExpressionScalarParameter_51 {
   class=MaterialExpressionScalarParameter
   properties {defaultValue=0.033, group=Wrap_NebulaV2_Galaxy, parameterName=Neb_Galaxy Rotation Speed}
  }
  MaterialExpressionScalarParameter_52 {
   class=MaterialExpressionScalarParameter
   properties {defaultValue=-0.03, group=Wrap_NebulaV2_Galaxy, parameterName=Neb_Galaxy Rotation Speed_1}
  }
  MaterialExpressionScalarParameter_53 {
   class=MaterialExpressionScalarParameter
   properties {defaultValue=0.05, group=Wrap_NebulaV2_Galaxy, parameterName=Neb_Galaxy Rotation Speed_2}
  }
  MaterialExpressionScalarParameter_54 {class=MaterialExpressionScalarParameter, properties {defaultValue=0.1, group=Wrap_NebulaV2_Galaxy, parameterName=Neb_Background Rotation}}
  MaterialExpressionScalarParameter_55 {class=MaterialExpressionScalarParameter, properties {defaultValue=0.5, group=Wrap_NebulaV2_Blend, parameterName=Neb_Blend_Glow_Warp_Amount}}
  MaterialExpressionScalarParameter_56 {
   class=MaterialExpressionScalarParameter
   properties {group=Wrap_NebulaV2_Uvs_Horn, parameterName=Neb_Horn_Rotation_Vehicle, sortPriority=150}
  }
  MaterialExpressionScalarParameter_57 {
   class=MaterialExpressionScalarParameter
   properties {group=Wrap_NebulaV2_Uvs_Horn, parameterName=Neb_Horn_Rotation_Vehicle_Side, sortPriority=150}
  }
  MaterialExpressionScalarParameter_58 {
   class=MaterialExpressionScalarParameter
   properties {group=Wrap_NebulaV2_Uvs_Horn, parameterName="Neb_Horn_Rotation_Weapon\\Pickaxe", sortPriority=150}
  }
  MaterialExpressionScalarParameter_59 {
   class=MaterialExpressionScalarParameter
   properties {group=Wrap_NebulaV2_Uvs_Horn, parameterName="Neb_Horn_Rotation_Mech\\Character\\Backpack", sortPriority=150}
  }
  MaterialExpressionScalarParameter_6 {class=MaterialExpressionScalarParameter, properties {defaultValue=3, group=Wrap_NebulaV2_Galaxy, parameterName=Neb_Edge_Strength}}
  MaterialExpressionScalarParameter_60 {class=MaterialExpressionScalarParameter, properties {defaultValue=1, group=Wrap_NebulaV2_Horn, parameterName=Neb_Horn_Tri_Blend_Contrast}}
  MaterialExpressionScalarParameter_61 {class=MaterialExpressionScalarParameter, properties {defaultValue=160, group=Wrap_NebulaV2_Galaxy_UV, parameterName=Neb_Scale_Gal_Vehicle}}
  MaterialExpressionScalarParameter_62 {class=MaterialExpressionScalarParameter, properties {defaultValue=6, group=Wrap_NebulaV2_Galaxy, parameterName=Neb_GalaxyAxisFade}}
  MaterialExpressionScalarParameter_63 {class=MaterialExpressionScalarParameter, properties {defaultValue=2.5, group=Wrap_NebulaV2, parameterName=Neb_BC_Edge_Brightness}}
  MaterialExpressionScalarParameter_64 {class=MaterialExpressionScalarParameter, properties {defaultValue=2.5, group=Wrap_NebulaV2, parameterName=Neb_BC_Edge_Brightness}}
  MaterialExpressionScalarParameter_65 {class=MaterialExpressionScalarParameter, properties {group=Wrap_NebulaV2_Galaxy, parameterName=Neb_EdgeEffect_Msk_BkGround}}
  MaterialExpressionScalarParameter_66 {
   class=MaterialExpressionScalarParameter
   properties {defaultValue=1, group=Wrap_NebulaV2_Galaxy, parameterName=Neb_EdgeEffect_Msk_BkGroun_Cont}
  }
  MaterialExpressionScalarParameter_67 {
   class=MaterialExpressionScalarParameter
   properties {defaultValue=0.33, group=Wrap_NebulaV2_Galaxy, parameterName=Neb_GalaxyDesaturationAmount}
  }
  MaterialExpressionScalarParameter_68 {class=MaterialExpressionScalarParameter, properties {defaultValue=-0.5, parameterName=Neb_Grad_Offset}}
  MaterialExpressionScalarParameter_69 {
   class=MaterialExpressionScalarParameter
   properties {defaultValue=1, group=Wrap_NebulaV2_Blend, parameterName="Neb_Blend_Mask_Amount  1=off"}
  }
  MaterialExpressionScalarParameter_7 {
   class=MaterialExpressionScalarParameter
   properties {defaultValue=-0.5, group=Wrap_NebulaV2_Grad, parameterName=Neb_Grad_Offset_Pickaxe, sortPriority=150}
  }
  MaterialExpressionScalarParameter_70 {
   class=MaterialExpressionScalarParameter
   properties {defaultValue=1, group=Wrap_NebulaV2, parameterName="Neb_Mask_Primary   0=On", sliderMax=1, sortPriority=0}
  }
  MaterialExpressionScalarParameter_71 {
   class=MaterialExpressionScalarParameter
   properties {defaultValue=1, group=Wrap_NebulaV2, parameterName="Neb_Mask_Accent   0=On", sliderMax=1, sortPriority=0}
  }
  MaterialExpressionScalarParameter_72 {
   class=MaterialExpressionScalarParameter
   properties {defaultValue=1, group=Wrap_NebulaV2, parameterName="Neb_Mask__Everthing   1=On", sliderMax=1, sortPriority=0}
  }
  MaterialExpressionScalarParameter_73 {
   class=MaterialExpressionScalarParameter
   properties {defaultValue=160, group=Wrap_NebulaV2_Galaxy_UV, parameterName=Neb_Scale_Gal_Vehicle_Side}
  }
  MaterialExpressionScalarParameter_74 {class=MaterialExpressionScalarParameter, properties {defaultValue=160, group=Wrap_NebulaV2_Galaxy_UV, parameterName=Neb_Scale_Gal_Weapon}}
  MaterialExpressionScalarParameter_75 {class=MaterialExpressionScalarParameter, properties {defaultValue=160, group=Wrap_NebulaV2_Galaxy_UV, parameterName=Neb_Scale_Gal_Warp_Mech}}
  MaterialExpressionScalarParameter_76 {class=MaterialExpressionScalarParameter, properties {defaultValue=160, group=Wrap_NebulaV2_Galaxy_UV, parameterName=Neb_Scale_Gal_Backpack}}
  MaterialExpressionScalarParameter_77 {class=MaterialExpressionScalarParameter, properties {defaultValue=160, group=Wrap_NebulaV2_Galaxy_UV, parameterName=Neb_Scale_Gal_Pickaxe}}
  MaterialExpressionScalarParameter_78 {class=MaterialExpressionScalarParameter, properties {defaultValue=160, group=Wrap_NebulaV2_Galaxy_UV, parameterName=Neb_Scale_Gal_Character}}
  MaterialExpressionScalarParameter_79 {class=MaterialExpressionScalarParameter, properties {group=Wrap_NebulaV2_Galaxy, parameterName=Neb_EdgeEffect_Msk}}
  MaterialExpressionScalarParameter_8 {
   class=MaterialExpressionScalarParameter
   properties {defaultValue=-0.5, group=Wrap_NebulaV2_Grad, parameterName=Neb_Grad_Offset_Character, sortPriority=150}
  }
  MaterialExpressionScalarParameter_80 {class=MaterialExpressionScalarParameter, properties {defaultValue=-0.1, group=Wrap_NebulaV2_Galaxy, parameterName=Neb_Edge_MinClamp}}
  MaterialExpressionScalarParameter_81 {class=MaterialExpressionScalarParameter, properties {group=Wrap_NebulaV2_Grad, parameterName=Neb_Msk_Offset_Vehicle, sortPriority=150}}
  MaterialExpressionScalarParameter_82 {class=MaterialExpressionScalarParameter, properties {group=Wrap_NebulaV2_Grad, parameterName=Neb_Msk_Offset_Vehicle_Side, sortPriority=150}}
  MaterialExpressionScalarParameter_83 {class=MaterialExpressionScalarParameter, properties {group=Wrap_NebulaV2_Grad, parameterName=Neb_Msk_Offset_Weapon, sortPriority=150}}
  MaterialExpressionScalarParameter_84 {class=MaterialExpressionScalarParameter, properties {group=Wrap_NebulaV2_Grad, parameterName=Neb_Msk_Offset_Mech, sortPriority=150}}
  MaterialExpressionScalarParameter_85 {class=MaterialExpressionScalarParameter, properties {group=Wrap_NebulaV2_Grad, parameterName=Neb_Msk_Offset_Backpack, sortPriority=150}}
  MaterialExpressionScalarParameter_86 {class=MaterialExpressionScalarParameter, properties {defaultValue=1, group=Wrap_NebulaV2_Blend, parameterName=Neb_Fill_Alternate_Axies}}
  MaterialExpressionScalarParameter_87 {class=MaterialExpressionScalarParameter, properties {defaultValue=1, group=Wrap_NebulaV2_Blend, parameterName=Neb_Fill_Alternate_Axies_Veh}}
  MaterialExpressionScalarParameter_88 {class=MaterialExpressionScalarParameter, properties {group=Wrap_NebulaV2_Grad, parameterName=Neb_Msk_Offset_Pickaxe, sortPriority=150}}
  MaterialExpressionScalarParameter_89 {class=MaterialExpressionScalarParameter, properties {group=Wrap_NebulaV2_Grad, parameterName=Neb_Msk_Offset_Character, sortPriority=150}}
  MaterialExpressionScalarParameter_9 {class=MaterialExpressionScalarParameter, properties {group=Wrap_NebulaV2, parameterName=Neb_Metalic}}
  MaterialExpressionScalarParameter_90 {class=MaterialExpressionScalarParameter, properties {group=Wrap_NebulaV2, parameterName=Neb_Spec}}
  MaterialExpressionScalarParameter_91 {
   class=MaterialExpressionScalarParameter
   properties {
    desc="1 = will sidable MF default blend mask  TS_Emissive needs to be active to work "
    group=Wrap_NebulaV2
    parameterName=Neb_Mask_TS_Emissive_01
    sortPriority=0
   }
  }
  MaterialExpressionScalarParameter_92 {
   class=MaterialExpressionScalarParameter
   properties {
    desc="1 = will sidable MF default blend mask  TS_Emissive needs to be active to work "
    group=Wrap_NebulaV2
    parameterName=Neb_Mask_TS_Emissive_02
    sortPriority=0
   }
  }
  MaterialExpressionScalarParameter_93 {class=MaterialExpressionScalarParameter, properties {group=Wrap_NebulaV2, parameterName="Neb_Root_Emissive 1=on", sortPriority=0}}
  MaterialExpressionScalarParameter_94 {
   class=MaterialExpressionScalarParameter
   properties {
    desc="1 = will sidable MF default blend mask  TS_Emissive needs to be active to work "
    group=Wrap_NebulaV2
    parameterName="Neb_Mask 0=Nebula 1=TS01 EmisMsk  2=TS02 EmisMsk"
    sortPriority=0
   }
  }
  MaterialExpressionScreenPosition_3 {class=MaterialExpressionScreenPosition, inputs [], properties {}}
  MaterialExpressionSetMaterialAttributes_0 {
   class=MaterialExpressionSetMaterialAttributes
   inputs [
    MaterialAttributes {from=MaterialExpressionNamedRerouteUsage_100, output=""}
    BaseColor {from=MaterialExpressionConstant_13, output=""}
    Roughness {from=MaterialExpressionScalarParameter_29, output=""}
    Specular {from=MaterialExpressionLinearInterpolate_20, output=""}
    Metallic {from=MaterialExpressionConstant_13, output=""}
    Normal {from="", output=""}
    EmissiveColor {from=MaterialExpressionAdd_25, output=""}
   ]
   properties {attributeSetTypes [
     #1
     #2
     #6
     #5
     #3
     #4
    ]}
  }
  MaterialExpressionSetMaterialAttributes_1 {
   class=MaterialExpressionSetMaterialAttributes
   inputs [
    MaterialAttributes {from="", output=""}
    Normal {from=MaterialExpressionTextureSample_4, output=RGB}
    BaseColor {from=MaterialExpressionConstant3Vector_1, output=""}
   ]
   properties {attributeSetTypes [#3, #1]}
  }
  MaterialExpressionSetMaterialAttributes_2 {
   class=MaterialExpressionSetMaterialAttributes
   inputs [
    MaterialAttributes {from=MaterialExpressionNamedRerouteUsage_4, output=""}
    BaseColor {from=MaterialExpressionLinearInterpolate_55, output=""}
    Roughness {from=MaterialExpressionScalarParameter_2, output=""}
    Specular {from=MaterialExpressionScalarParameter_3, output=""}
    Metallic {from=MaterialExpressionScalarParameter_4, output=""}
    Normal {from="", output=""}
    EmissiveColor {from="", output=""}
   ]
   properties {attributeSetTypes [
     #1
     #2
     #6
     #5
     #3
     #4
    ]}
  }
  MaterialExpressionSetMaterialAttributes_3 {
   class=MaterialExpressionSetMaterialAttributes
   inputs [
    MaterialAttributes {from=MaterialExpressionNamedRerouteUsage_64, output=""}
    BaseColor {from=MaterialExpressionLinearInterpolate_54, output=""}
    Roughness {from=MaterialExpressionNamedRerouteUsage_45, output=""}
    Specular {from=MaterialExpressionScalarParameter_30, output=""}
    Metallic {from=MaterialExpressionNamedRerouteUsage_6, output=""}
    Normal {from=MaterialExpressionAdd_5, output=""}
    EmissiveColor {from="", output=""}
   ]
   properties {attributeSetTypes [
     #1
     #2
     #6
     #5
     #3
     #4
    ]}
  }
  MaterialExpressionSetMaterialAttributes_4 {
   class=MaterialExpressionSetMaterialAttributes
   inputs [
    MaterialAttributes {from=MaterialExpressionNamedRerouteUsage_98, output=""}
    BaseColor {from=MaterialExpressionLinearInterpolate_54, output=""}
    Roughness {from="", output=""}
    Specular {from="", output=""}
    Metallic {from="", output=""}
    Normal {from="", output=""}
    EmissiveColor {from="", output=""}
   ]
   properties {attributeSetTypes [
     #1
     #2
     #6
     #5
     #3
     #4
    ]}
  }
  MaterialExpressionShadingPathSwitch_0 {
   class=MaterialExpressionShadingPathSwitch
   inputs [
    Default {from=MaterialExpressionVectorParameter_13, output=RGB}
    Deferred {from="", output=""}
    Forward {from="", output=""}
    Mobile {from=MaterialExpressionVectorParameter_18, output=RGB}
   ]
   properties {}
  }
  MaterialExpressionShadingPathSwitch_1 {
   class=MaterialExpressionShadingPathSwitch
   inputs [
    Default {from=MaterialExpressionLinearInterpolate_35, output=""}
    Deferred {from="", output=""}
    Forward {from="", output=""}
    Mobile {from=MaterialExpressionLinearInterpolate_34, output=""}
   ]
   properties {}
  }
  MaterialExpressionStaticBoolParameter_0 {
   class=MaterialExpressionStaticBoolParameter
   inputs []
   properties {defaultValue=true, group=Wrap_NebulaV2, parameterName=Neb_Disable_Rotation}
  }
  MaterialExpressionStaticBool_0 {class=MaterialExpressionStaticBool, properties {value=true}}
  MaterialExpressionStaticSwitchParameter_0 {
   class=MaterialExpressionStaticSwitchParameter
   inputs [
    True {from=MaterialExpressionNamedRerouteUsage_104, output=""}
    False {from=MaterialExpressionBlendMaterialAttributes_1, output=""}
   ]
   properties {
    desc=Uses the Top Right corner of the B channel of the _CM Mask to define a simple secondary material
    group=Customization
    parameterName="Allow Secondary Customization Area (Simple)"
    sortPriority=-499
   }
  }
  MaterialExpressionStaticSwitchParameter_1 {
   class=MaterialExpressionStaticSwitchParameter
   inputs [
    True {from=MaterialExpressionAdd_14, output=""}
    False {from=MaterialExpressionAdd_15, output=""}
   ]
   properties {group=Customization, parameterName=Function_Left_Right_Decal}
  }
  MaterialExpressionStaticSwitchParameter_2 {
   class=MaterialExpressionStaticSwitchParameter
   inputs [
    True {from=MaterialExpressionFunctionInput_18, output=""}
    False {from=MaterialExpressionAdd_23, output=""}
   ]
   properties {
    desc=Uses the Top Right corner of the B channel of the _CM Mask to define a simple secondary material
    group=Customization
    parameterName="Allow Secondary Customization Area (Simple)"
    sortPriority=-499
   }
  }
  MaterialExpressionStaticSwitchParameter_33 {
   class=MaterialExpressionStaticSwitchParameter
   inputs [
    True {from=MaterialExpressionBlendMaterialAttributes_2, output=""}
    False {from=MaterialExpressionNamedRerouteUsage_102, output=""}
   ]
   properties {group=Customization_Wrap, parameterName=Wrap_Nebula}
  }
  MaterialExpressionSubtract_0 {
   class=MaterialExpressionSubtract
   inputs [
    A {from=MaterialExpressionVertexInterpolator_4, output=PS}
    B {from=MaterialExpressionPreSkinnedLocalBounds_1, output=Min}
   ]
  }
  MaterialExpressionSubtract_3 {
   class=MaterialExpressionSubtract
   inputs [
    A {from=MaterialExpressionPreSkinnedLocalBounds_1, output=Max}
    B {from=MaterialExpressionPreSkinnedLocalBounds_1, output=Max}
   ]
  }
  MaterialExpressionSubtract_4 {
   class=MaterialExpressionSubtract
   inputs [
    A {from=MaterialExpressionPreSkinnedLocalBounds_3, output=Max}
    B {from=MaterialExpressionPreSkinnedLocalBounds_3, output=Max}
   ]
  }
  MaterialExpressionSwitch_0 {
   class=MaterialExpressionSwitch
   inputs [
    SwitchValue {from=MaterialExpressionScalarParameter_94, output=""}
    Default {from="", output=""}
    None {from=MaterialExpressionNamedRerouteUsage_69, output=""}
    None {from=MaterialExpressionFunctionInput_5, output=""}
    None {from=MaterialExpressionFunctionInput_6, output=""}
   ]
   properties {
    description="Neb_Mask_TS   1=TS01  2=TS02"
    inputs [
     {
      inputName=None
      input {
       expression=&self:MaterialExpressionNamedRerouteUsage_69
       inputName=None
       mask=0
       maskA=0
       maskB=0
       maskG=0
       maskR=0
       outputIndex=0
      }
     }
     {
      inputName=None
      input {
       expression=&self:MaterialExpressionFunctionInput_5
       inputName=None
       mask=0
       maskA=0
       maskB=0
       maskG=0
       maskR=0
       outputIndex=0
      }
     }
     {
      inputName=None
      input {
       expression=&self:MaterialExpressionFunctionInput_6
       inputName=None
       mask=0
       maskA=0
       maskB=0
       maskG=0
       maskR=0
       outputIndex=0
      }
     }
    ]
   }
  }
  MaterialExpressionTextureCoordinate_7 {class=MaterialExpressionTextureCoordinate, properties {uTiling=2, vTiling=2}}
  MaterialExpressionTextureObjectParameter_0 {
   class=MaterialExpressionTextureObjectParameter
   inputs [
    UVs {from="", output=""}
    Apply View MipBias {from="", output=""}
   ]
   properties {group=Wrap_NebulaV2, parameterName=Neb_Tex_Star_Msk, sortPriority=5, texture=&T_BlacksRGB}
  }
  MaterialExpressionTextureObjectParameter_1 {
   class=MaterialExpressionTextureObjectParameter
   inputs [
    UVs {from="", output=""}
    Apply View MipBias {from="", output=""}
   ]
   properties {group=Wrap_NebulaV2, parameterName=Neb_Tex_MSk_Blend, sortPriority=5, texture=&T_BlacksRGB}
  }
  MaterialExpressionTextureObjectParameter_2 {
   class=MaterialExpressionTextureObjectParameter
   inputs [
    UVs {from="", output=""}
    Apply View MipBias {from="", output=""}
   ]
   properties {group=Wrap_NebulaV2, parameterName=Neb_Tex_BC, sortPriority=5, texture=&T_BlacksRGB}
  }
  MaterialExpressionTextureObjectParameter_3 {
   class=MaterialExpressionTextureObjectParameter
   inputs [
    UVs {from="", output=""}
    Apply View MipBias {from="", output=""}
   ]
   properties {
    group=Wrap_NebulaV2
    parameterName=Neb_Tex_Nrml
    samplerType=SAMPLERTYPE_Normal
    sortPriority=5
    texture=&FlatNormalBlue
   }
  }
  MaterialExpressionTextureObjectParameter_4 {
   class=MaterialExpressionTextureObjectParameter
   inputs [
    UVs {from="", output=""}
    Apply View MipBias {from="", output=""}
   ]
   properties {group=Wrap_NebulaV2, parameterName=Neb_Tex_MSk1, sortPriority=5, texture=&T_BlacksRGB}
  }
  MaterialExpressionTextureObjectParameter_7 {
   class=MaterialExpressionTextureObjectParameter
   inputs [
    UVs {from="", output=""}
    Apply View MipBias {from="", output=""}
   ]
   properties {group=Wrap_NebulaV2, parameterName=Neb_Tex_Gal, sortPriority=5, texture=&T_BlacksRGB}
  }
  MaterialExpressionTextureSample_0 {
   class=MaterialExpressionTextureSample
   inputs [
    UVs {from=""}
    Tex {from=""}
    Apply View MipBias {from=""}
   ]
   properties {samplerType=SAMPLERTYPE_Masks, texture=/Game/Weapons/FORT_Rifles/Textures/T_Scar_CM}
  }
  MaterialExpressionTextureSample_1 {
   class=MaterialExpressionTextureSample
   inputs [
    UVs {from=MaterialExpressionNamedRerouteDeclaration_19}
    Tex {from=MaterialExpressionNamedRerouteUsage_103}
    Apply View MipBias {from=""}
   ]
   properties {samplerType=SAMPLERTYPE_Normal}
  }
  MaterialExpressionTextureSample_10 {
   class=MaterialExpressionTextureSample
   inputs [
    UVs {from=MaterialExpressionPanner_2}
    Tex {from=MaterialExpressionNamedRerouteUsage_30}
    Apply View MipBias {from=""}
   ]
   properties {samplerSource=SSM_Wrap_WorldGroupSettings, samplerType=SAMPLERTYPE_Masks}
  }
  MaterialExpressionTextureSample_11 {
   class=MaterialExpressionTextureSample
   inputs [
    UVs {from=MaterialExpressionNamedRerouteDeclaration_39}
    Tex {from=""}
    Apply View MipBias {from=""}
   ]
   properties {texture=/Game/Weapons/WeaponSkins/Wraps/Textures/UVCheaker}
  }
  MaterialExpressionTextureSample_12 {
   class=MaterialExpressionTextureSample
   inputs [
    UVs {from=MaterialExpressionAdd_6}
    Tex {from=""}
    Apply View MipBias {from=""}
   ]
   properties {texture=/Game/Weapons/WeaponSkins/Wraps/Textures/UVCheaker_Clamped}
  }
  MaterialExpressionTextureSample_2 {
   class=MaterialExpressionTextureSample
   inputs [
    UVs {from=MaterialExpressionSaturate_8}
    Tex {from=MaterialExpressionNamedRerouteUsage_81}
    Apply View MipBias {from=""}
   ]
   properties {samplerType=SAMPLERTYPE_Masks}
  }
  MaterialExpressionTextureSample_3 {
   class=MaterialExpressionTextureSample
   inputs [
    UVs {from=""}
    Tex {from=""}
    Apply View MipBias {from=""}
   ]
   properties {samplerType=SAMPLERTYPE_Masks, texture=&T_Scar_CM}
  }
  MaterialExpressionTextureSample_4 {
   class=MaterialExpressionTextureSample
   inputs [
    UVs {from=""}
    Tex {from=""}
    Apply View MipBias {from=""}
   ]
   properties {samplerType=SAMPLERTYPE_Normal, texture=/Game/Weapons/FORT_Rifles/T_SCAR_N}
  }
  MaterialExpressionTextureSample_5 {
   class=MaterialExpressionTextureSample
   inputs [
    UVs {from=MaterialExpressionNamedRerouteDeclaration_19}
    Tex {from=MaterialExpressionNamedRerouteUsage_2}
    Apply View MipBias {from=""}
   ]
   properties {samplerType=SAMPLERTYPE_Normal}
  }
  MaterialExpressionTextureSample_6 {
   class=MaterialExpressionTextureSample
   inputs [
    UVs {from=MaterialExpressionPanner_0}
    Tex {from=MaterialExpressionNamedRerouteUsage_7}
    Apply View MipBias {from=""}
   ]
  }
  MaterialExpressionTextureSample_7 {
   class=MaterialExpressionTextureSample
   inputs [
    UVs {from=MaterialExpressionPanner_1}
    Tex {from=MaterialExpressionNamedRerouteUsage_28}
    Apply View MipBias {from=""}
   ]
  }
  MaterialExpressionTextureSample_8 {
   class=MaterialExpressionTextureSample
   inputs [
    UVs {from=MaterialExpressionPanner_3}
    Tex {from=MaterialExpressionNamedRerouteUsage_12}
    Apply View MipBias {from=""}
   ]
   properties {
    samplerSource=SSM_Wrap_WorldGroupSettings
    samplerType=SAMPLERTYPE_Masks
    texture=/Game/Characters/Player/Male/Medium/Bodies/M_Med_Soldier_04/Skins/Galaxy/Textures/Celestial_T_Stars
   }
  }
  MaterialExpressionTextureSample_9 {
   class=MaterialExpressionTextureSample
   inputs [
    UVs {from=MaterialExpressionPanner_4}
    Tex {from=MaterialExpressionNamedRerouteUsage_20}
    Apply View MipBias {from=""}
   ]
  }
  MaterialExpressionTime_1 {class=MaterialExpressionTime}
  MaterialExpressionTime_5 {class=MaterialExpressionTime}
  MaterialExpressionTime_7 {class=MaterialExpressionTime}
  MaterialExpressionTime_8 {class=MaterialExpressionTime}
  MaterialExpressionTransformPosition_0 {
   class=MaterialExpressionTransformPosition
   inputs [None {from=MaterialExpressionComponentMask_7, output=""}]
   properties {transformType=TRANSFORMPOSSOURCE_World}
  }
  MaterialExpressionVectorParameter_0 {
   class=MaterialExpressionVectorParameter
   properties {group=Wrap_NebulaV2_Accent, parameterName=Neb_Accent_BaseColor, defaultValue=<rgba 0.01 0.01 0.01 1>}
  }
  MaterialExpressionVectorParameter_1 {class=MaterialExpressionVectorParameter, properties {group=Wrap_NebulaV2_Uvs, parameterName=Neb_Offset_Weapon, defaultValue=<rgba 0 0 0 1>}}
  MaterialExpressionVectorParameter_10 {
   class=MaterialExpressionVectorParameter
   properties {group=Wrap_NebulaV2_Galaxy, parameterName=Neb_Small Star_Pan_Speed, defaultValue=<rgba 0 0 0 1>}
  }
  MaterialExpressionVectorParameter_11 {class=MaterialExpressionVectorParameter, properties {group=Wrap_NebulaV2_Uvs, parameterName=Neb_Offset_Mech, defaultValue=<rgba 0 0 0 1>}}
  MaterialExpressionVectorParameter_12 {class=MaterialExpressionVectorParameter, properties {group=Wrap_NebulaV2_Uvs, parameterName=Neb_Offset_Vehicle, defaultValue=<rgba 0 0 0 1>}}
  MaterialExpressionVectorParameter_13 {
   class=MaterialExpressionVectorParameter
   properties {group=Wrap_NebulaV2_Galaxy, parameterName=Neb_BGAddTint, defaultValue=<rgba 0.0112 0.0138 0.02 1>}
  }
  MaterialExpressionVectorParameter_14 {
   class=MaterialExpressionVectorParameter
   properties {group=Wrap_NebulaV2_Uvs, parameterName=Neb_Offset_Character, defaultValue=<rgba 0 0 0 1>}
  }
  MaterialExpressionVectorParameter_15 {class=MaterialExpressionVectorParameter, properties {group=Wrap_NebulaV2_Uvs, parameterName=Neb_Offset_Pickaxe, defaultValue=<rgba 0 0 0 1>}}
  MaterialExpressionVectorParameter_16 {
   class=MaterialExpressionVectorParameter
   properties {group=Wrap_NebulaV2_Blend, parameterName=Neb_Blend_Glow_Color_B, defaultValue=<rgba 0.0663 0.0665 0.1667 1>}
  }
  MaterialExpressionVectorParameter_17 {
   class=MaterialExpressionVectorParameter
   properties {group=Wrap_NebulaV2_Uvs_Horn, parameterName=Neb_Horn_Offset_Scale_Character, defaultValue=<rgba 0 0 0 1>}
  }
  MaterialExpressionVectorParameter_18 {
   class=MaterialExpressionVectorParameter
   properties {group=Wrap_NebulaV2_Galaxy, parameterName=Neb_BGAddTint_Mobile, defaultValue=<rgba 0.0055 0.0057 0.01 1>}
  }
  MaterialExpressionVectorParameter_19 {
   class=MaterialExpressionVectorParameter
   properties {group=Wrap_NebulaV2_Uvs_Horn, parameterName=Neb_Horn_Offset_Scale_Weapon, defaultValue=<rgba -0.385 -0.02 0 0.3>}
  }
  MaterialExpressionVectorParameter_2 {
   class=MaterialExpressionVectorParameter
   properties {group=Wrap_NebulaV2_Uvs, parameterName=Neb_Offset_VehicleSide, defaultValue=<rgba 0 0 0 1>}
  }
  MaterialExpressionVectorParameter_20 {
   class=MaterialExpressionVectorParameter
   properties {group=Wrap_NebulaV2_Uvs_Horn, parameterName=Neb_Horn_Offset_Scale_Mech, defaultValue=<rgba 0 0 0 1>}
  }
  MaterialExpressionVectorParameter_21 {
   class=MaterialExpressionVectorParameter
   properties {group=Wrap_NebulaV2_Uvs_Horn, parameterName=Neb_Horn_Offset_Scale_Vehicle, defaultValue=<rgba 0 0 0 1>}
  }
  MaterialExpressionVectorParameter_22 {
   class=MaterialExpressionVectorParameter
   properties {group=Wrap_NebulaV2_Uvs_Horn, parameterName=Neb_Horn_Offset_Scale_VehicleSide, defaultValue=<rgba 0 0 0 1>}
  }
  MaterialExpressionVectorParameter_23 {
   class=MaterialExpressionVectorParameter
   properties {group=Wrap_NebulaV2_Galaxy, parameterName=Neb_BGMultTint, defaultValue=<rgba 0.1095 0.2307 0.5271 1>}
  }
  MaterialExpressionVectorParameter_24 {
   class=MaterialExpressionVectorParameter
   properties {group=Wrap_NebulaV2_Uvs_Horn, parameterName=Neb_Horn_Offset_Scale_Pickaxe, defaultValue=<rgba 0 0 0 1>}
  }
  MaterialExpressionVectorParameter_25 {
   class=MaterialExpressionVectorParameter
   properties {group=Wrap_NebulaV2_Uvs_Horn, parameterName=Neb_Horn_Offset_Scale_Backpack, defaultValue=<rgba 0 0 0 1>}
  }
  MaterialExpressionVectorParameter_26 {
   class=MaterialExpressionVectorParameter
   properties {group=Wrap_NebulaV2_Uvs, parameterName=Neb_TS_Rotation_Vehicle, sortPriority=200, defaultValue=<rgba 0 0 0 1>}
  }
  MaterialExpressionVectorParameter_27 {
   class=MaterialExpressionVectorParameter
   properties {group=Wrap_NebulaV2_Uvs, parameterName="Neb_TS_Rotation_Mech\\Character\\BackPack", sortPriority=200, defaultValue=<rgba 0.25 0.25 0 1>}
  }
  MaterialExpressionVectorParameter_28 {
   class=MaterialExpressionVectorParameter
   properties {group=Wrap_NebulaV2_Horn, parameterName=Neb_Horn_Glow_Color_A, defaultValue=<rgba 0 0.2313 1 1>}
  }
  MaterialExpressionVectorParameter_29 {
   class=MaterialExpressionVectorParameter
   properties {group=Wrap_NebulaV2_Horn, parameterName=Neb_Horn_Glow_Color_B, defaultValue=<rgba 0.9342 1.9203 2 1>}
  }
  MaterialExpressionVectorParameter_3 {
   class=MaterialExpressionVectorParameter
   properties {group=Wrap_NebulaV2_Galaxy, parameterName=Neb_Galaxy_PanSpeed, defaultValue=<rgba 0.005 0.015 0 0>}
  }
  MaterialExpressionVectorParameter_30 {
   class=MaterialExpressionVectorParameter
   properties {group=Wrap_NebulaV2_Blend, parameterName=Neb_Blend_Glow_Color_C, defaultValue=<rgba 0.1522 0.056 0.2344 1>}
  }
  MaterialExpressionVectorParameter_31 {
   class=MaterialExpressionVectorParameter
   properties {group=Wrap_NebulaV2_Galaxy, parameterName=Neb_Small Stars_PanSpeed, defaultValue=<rgba -0.02 0.02 0 0>}
  }
  MaterialExpressionVectorParameter_32 {
   class=MaterialExpressionVectorParameter
   properties {group=Wrap_NebulaV2_Blend, parameterName=Neb_Blend_Glow_Color_Ab, defaultValue=<rgba 0.2879 0.1627 0.2969 1>}
  }
  MaterialExpressionVectorParameter_33 {
   class=MaterialExpressionVectorParameter
   properties {group=Wrap_NebulaV2_Blend, parameterName=Neb_Blend_Glow_Color_AC, defaultValue=<rgba 0 0.9635 0.1305 1>}
  }
  MaterialExpressionVectorParameter_34 {
   class=MaterialExpressionVectorParameter
   properties {group=Wrap_NebulaV2_Galaxy, parameterName=Neb_Galaxy_Fres_Color, sortPriority=-10, defaultValue=<rgba 1 1 1 1>}
  }
  MaterialExpressionVectorParameter_35 {
   class=MaterialExpressionVectorParameter
   properties {group=Wrap_NebulaV2_Galaxy, parameterName=Neb_Galaxy_BaseColor, sortPriority=-10, defaultValue=<rgba 0 0 0 1>}
  }
  MaterialExpressionVectorParameter_36 {
   class=MaterialExpressionVectorParameter
   properties {group=Wrap_NebulaV2_Uvs, parameterName=Neb_Offset_Backpack, defaultValue=<rgba 0 0 0 1>}
  }
  MaterialExpressionVectorParameter_37 {
   class=MaterialExpressionVectorParameter
   properties {group=Wrap_NebulaV2_Uvs, parameterName=Neb_Scale_Warp_Weapon, defaultValue=<rgba 1 1 0 3>}
  }
  MaterialExpressionVectorParameter_38 {
   class=MaterialExpressionVectorParameter
   properties {group=Wrap_NebulaV2_Uvs, parameterName=Neb_Scale_Warp_Vehicle, defaultValue=<rgba 1 1 0 3>}
  }
  MaterialExpressionVectorParameter_39 {
   class=MaterialExpressionVectorParameter
   properties {group=Wrap_NebulaV2_Uvs, parameterName=Neb_Scale_Warp_Vehicle_Side, defaultValue=<rgba 1 1 0 3>}
  }
  MaterialExpressionVectorParameter_4 {
   class=MaterialExpressionVectorParameter
   properties {group=Wrap_NebulaV2_Uvs, parameterName="Neb_TS_Rotation_Weapon\\PickAxe", sortPriority=200, defaultValue=<rgba 0 0 -0.25 1>}
  }
  MaterialExpressionVectorParameter_40 {
   class=MaterialExpressionVectorParameter
   properties {group=Wrap_NebulaV2_Uvs, parameterName=Neb_Scale_Warp_Character, defaultValue=<rgba 1 1 0 3>}
  }
  MaterialExpressionVectorParameter_41 {
   class=MaterialExpressionVectorParameter
   properties {group=Wrap_NebulaV2_Uvs, parameterName=Neb_Scale_Warp_Mech, defaultValue=<rgba 1 1 0 3>}
  }
  MaterialExpressionVectorParameter_42 {
   class=MaterialExpressionVectorParameter
   properties {group=Wrap_NebulaV2_Uvs, parameterName=Neb_Scale_Warp_Backpack, defaultValue=<rgba 1 1 0 3>}
  }
  MaterialExpressionVectorParameter_43 {
   class=MaterialExpressionVectorParameter
   properties {group=Wrap_NebulaV2_Uvs, parameterName=Neb_Scale_Warp_Pickaxe, defaultValue=<rgba 1 1 0 3>}
  }
  MaterialExpressionVectorParameter_44 {
   class=MaterialExpressionVectorParameter
   properties {group=Wrap_NebulaV2_Galaxy, parameterName=Neb_Bright Star_Msk_PanSpeed, defaultValue=<rgba 0.133 0.2 0 0>}
  }
  MaterialExpressionVectorParameter_45 {class=MaterialExpressionVectorParameter, properties {group=Wrap_NebulaV2_Galaxy, parameterName=Neb_Bright Stars_PanSpeed}}
  MaterialExpressionVectorParameter_5 {class=MaterialExpressionVectorParameter, properties {group=Wrap_NebulaV2_Blend, parameterName=Neb_Fill_Alternate_Msk}}
  MaterialExpressionVectorParameter_58 {
   class=MaterialExpressionVectorParameter
   properties {group=Wrap_NebulaV2_Uvs, parameterName=Neb_TS_Rotation_Vehicle_Side, sortPriority=200, defaultValue=<rgba 0 0 0 1>}
  }
  MaterialExpressionVectorParameter_6 {
   class=MaterialExpressionVectorParameter
   properties {group=Wrap_NebulaV2_Blend, parameterName=Neb_Blend_Warp_A_Speed, defaultValue=<rgba 0.5 0.05 0 1>}
  }
  MaterialExpressionVectorParameter_7 {
   class=MaterialExpressionVectorParameter
   properties {group=Wrap_NebulaV2_Blend, parameterName=Neb_Blend_Glow_Color_A, defaultValue=<rgba 0.9844 0.0715 0.0911 1>}
  }
  MaterialExpressionVectorParameter_8 {
   class=MaterialExpressionVectorParameter
   properties {group=Wrap_NebulaV2_Galaxy, parameterName=Neb_Star Brightness, defaultValue=<rgba 50 50 50 1>}
  }
  MaterialExpressionVectorParameter_9 {
   class=MaterialExpressionVectorParameter
   properties {group=Wrap_NebulaV2_Galaxy, parameterName=Neb_Star Brightness_1, defaultValue=<rgba 200 200 200 1>}
  }
  MaterialExpressionVertexInterpolator_0 {class=MaterialExpressionVertexInterpolator, inputs [VS {from=MaterialExpressionPreSkinnedNormal_1, output=""}], properties {}}
  MaterialExpressionVertexInterpolator_4 {class=MaterialExpressionVertexInterpolator, inputs [VS {from=MaterialExpressionPreSkinnedPosition_0, output=""}], properties {}}
 }}
