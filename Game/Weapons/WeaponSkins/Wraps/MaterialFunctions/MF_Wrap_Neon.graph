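# Material function for the Neon weapon wrap, blended over the input material by the customization mask behind a static switch.
# It layers warped, hue-cycling emissive stripes over a rainbow background of projected grid, curve-atlas pulse wave and fresnel.
# Music/emote audio and a decaying Blueprint-timed trigger boost stripe warp and hue shift.
# An isState input selects tiling for weapon, vehicle, mech or character, and time of day scales the glow.
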MaterialFunction /Game/Weapons/WeaponSkins/Wraps/MaterialFunctions/MF_Wrap_Neon
graph {nodes [
  n1 = Subtract(A=Time(), B=ScalarParameter() {defaultValue=-9999999, group=Wrap_NeonJellyfish, parameterName=TimeFrom-BP----------------, sortPriority=600})
  n2 = ScalarParameter() {defaultValue=0.8, group=Wrap_Neon, parameterName=Neon_Time_Duration, sortPriority=601}
  n3 = OneMinus(None=Saturate(None=Divide(A=@n1, B=@n2)))
  n4 = CollectionParameter() {collection=&FortniteMaterialParameters, parameterName=EventAudio_Amplitude}
  n5 = CollectionParameter() {collection=&FortniteMaterialParameters, parameterName=Emote_AudioLow}
  n6 = CollectionParameter() {collection=&FortniteMaterialParameters, parameterName=Emote_AudioHigh}
  n7 = Max(A=@n4, B=Max(A=@n5, B=@n6))
  n8 = LinearInterpolate(
   A=ScalarParameter() {group=Wrap_Neon, parameterName=Neon_Warp_B Amount_Low}
   B=ScalarParameter() {defaultValue=1, group=Wrap_Neon, parameterName=Neon_Warp_B Amount_High}
   Alpha=Max(A=@n7, B=@n3)
  )
  Add(A=@n8)
  n9 = FunctionInput(Preview=Constant()) {inputName=isState, inputType=FunctionInput_Vector4, sortPriority=5}
  n10 = MaterialFunctionCall(Color=@n9) {materialFunction=&SplitComponents}
  n11 = LinearInterpolate(
   A=ScalarParameter() {defaultValue=1, group=Wrap_Neon, parameterName=Neon_BC_TileRate_Weapon}
   B=LinearInterpolate(
    A=ScalarParameter() {defaultValue=1, group=Wrap_Neon, parameterName=Neon_BC_TileRate_Vehicle}
    B=ScalarParameter() {defaultValue=1, group=Wrap_Neon, parameterName=Neon_BC_TileRate_VehicleSide}
    Alpha=@n10.G
   )
   Alpha=@n10.R
  )
  n12 = ScalarParameter() {defaultValue=1, group=Wrap_Neon, parameterName=Neon_BC_TileRate_Mech}
  n13 = VertexInterpolator(VS=PreSkinnedPosition())
  n14 = PreSkinnedLocalBounds()
  n15 = FunctionInput(Preview=Divide(A=Subtract(A=@n13.PS, B=@n14.Min), B=Subtract(A=@n14.Max, B=@n14.Max))) {bUsePreviewValueAsDefault=true, inputName=Calcated_Bounds, sortPriority=11}
  n16 = FunctionInput(Preview=VertexInterpolator(VS=PreSkinnedNormal()).PS) {bUsePreviewValueAsDefault=true, inputName=Pre-Skinned Local Normal, sortPriority=10}
  n17 = MaterialFunctionCall(Color=Constant3Vector() {constant=<rgba 0 -0.25 -0.25 1>}) {materialFunction=&SplitComponents}
  n18 = MaterialFunctionCall(
   Export Float 4=StaticBool() {value=true}
   TextureObject=TextureObjectParameter() {group=Wrap_Neon, parameterName=Neon_BC_Text, texture=&T_BlacksRGB}
   TextureSize=Constant() {r=1}
   WorldPosition=Multiply(A=LinearInterpolate(A=@n11, B=@n12, Alpha=@n10.B), B=@n15)
   World Space Normal=@n16
   RotZAxis=@n17.B
   RotXAxis=@n17.B
   RotYAxis=@n17.B
   ProjectionTransitionContrast=ScalarParameter() {defaultValue=1, group=Wrap_Neon, parameterName=Neon_Blend_Contrast, sortPriority=1}
  ) {materialFunction=&/WorldAlignedTexture}
  ComponentMask(None=@n18.XYZ Texture) {b=true, g=true, r=true}
  Constant()
  FunctionInput(Preview=TextureSample() {samplerType=SAMPLERTYPE_Masks, texture=/Game/Weapons/FORT_Rifles/Textures/T_Scar_CM}.G) {bUsePreviewValueAsDefault=true, inputName=Curavture, inputType=FunctionInput_Scalar, sortPriority=24}
  n19 = GetMaterialAttributes(None=FunctionInput() {inputName=Material, inputType=FunctionInput_MaterialAttributes, sortPriority=-20})
  n20 = Constant3Vector() {constant=<rgba 0 1 0 0>}
  n21 = MaterialFunctionCall!(
   Bounds=@n15
   ForwardAxis=LinearInterpolate(
    A=FunctionInput(Preview=@n20) {bUsePreviewValueAsDefault=true, inputName=Foward Axis, sortPriority=24}
    B=Constant3Vector() {constant=<rgba 0 1 0 0>}
    Alpha=ScalarParameter() {group=Wrap_Neon, parameterName=Neon_IsRoll, sortPriority=-1}
   )
   Offset=Constant()
   FallOff=Constant() {r=1}
  ) {materialFunction=&/MF_Wrap_Gradent_Simple}
  n22 = ScalarParameter() {defaultValue=1, group=Wrap_Neon, parameterName=Neon_Stripes_zGradent_TileRate_Weapon}
  n23 = MaterialFunctionCall(Float4=@n9) {materialFunction=&BreakOutFloat4Components}
  n24 = LinearInterpolate(
   A=ScalarParameter() {defaultValue=1, group=Wrap_Neon, parameterName=Neon_Stripes_zGradent_TileRate_Vehicle}
   B=ScalarParameter() {defaultValue=1, group=Wrap_Neon, parameterName=Neon_Stripes_zGradent_TileRate_VehicleSide}
   Alpha=@n23.G
  )
  n25 = ScalarParameter() {defaultValue=1, group=Wrap_Neon, parameterName=Neon_Stripes_zGradent_TileRate_Mech}
  n26 = ScalarParameter() {defaultValue=1, group=Wrap_Neon, parameterName=Neon_Stripes_zGradent_TileRate_Character}
  n27 = Multiply(
   A=@n21.Raw_Data
   B=LinearInterpolate(A=LinearInterpolate(A=@n22, B=@n24, Alpha=@n23.R), B=LinearInterpolate(A=@n25, B=@n26, Alpha=@n23.A), Alpha=@n23.B)
  )
  n28 = Multiply(A=Time(), B=ScalarParameter() {defaultValue=1, group=Wrap_Neon, parameterName=Neon_Background_Gradent_Speed})
  n29 = MaterialFunctionCall(Hue Shift Percentage=Frac(None=Add(A=@n27, B=@n28)), Texture=Constant3Vector() {constant=<rgba 1 0 0 0>}) {materialFunction=&HueShift}
  n30 = ComponentMask(None=@n18.XYZ Texture) {r=true}
  n31 = Multiply(A=@n30, B=ScalarParameter() {defaultValue=0.01, group=Wrap_Neon, parameterName=Neon_Background_Grid_Strength}) {constB=0.01, desc=strength of grid everywhere}
  n32 = ScalarParameter() {defaultValue=1, group=Wrap_Neon, parameterName=Neon_Gradent_Pulse_TileRate_Weapon}
  n33 = MaterialFunctionCall(Float4=@n9) {materialFunction=&BreakOutFloat4Components}
  n34 = LinearInterpolate(
   A=ScalarParameter() {defaultValue=1, group=Wrap_Neon, parameterName=Neon_Gradent_Pulse_TileRate_Vehicle}
   B=ScalarParameter() {defaultValue=1, group=Wrap_Neon, parameterName=Neon_Gradent_Pulse_TileRate_VehicleSide}
   Alpha=@n33.G
  )
  n35 = ScalarParameter() {defaultValue=1, group=Wrap_Neon, parameterName=Neon_Gradent_Pulse_TileRate_Mech}
  n36 = ScalarParameter() {defaultValue=1, group=Wrap_Neon, parameterName=Neon_Gradent_Pulse_TileRate_Character}
  n37 = Multiply(
   A=@n21.Raw_Data
   B=LinearInterpolate(A=LinearInterpolate(A=@n32, B=@n34, Alpha=@n33.R), B=LinearInterpolate(A=@n35, B=@n36, Alpha=@n33.A), Alpha=@n33.B)
  )
  n38 = Multiply(A=Time(), B=ScalarParameter() {defaultValue=-1, group=Wrap_Neon, parameterName=Neon_Gradent_Pulse_Speed})
  n39 = CurveAtlasRowParameter(CurveTime=Frac(None=Add(A=@n37, B=@n38))) {
   atlas=&AthenaCurveAtlas01
   curve=/Game/Characters/Player/Male/Medium/Bodies/M_MED_Neon/Materials/neon_body_curve
   defaultValue=220
   group=Wrap_Neon
   parameterName=Neon_Gradent_pulse_ColorCurve
  }
  n40 = Power(Base=@n39.G)
  n41 = Multiply(A=@n29.Result, B=Max(A=Add(A=@n31, B=@n40)) {constB=0.003}) {desc=coloured grid + wave}
  n42 = ScalarParameter() {defaultValue=5, group=Wrap_Neon, parameterName=Neon_Background_Fresnel_Exponentin}
  n43 = Multiply(
   A=@n29.Result
   B=MaterialFunctionCall!(Base=Fresnel(ExponentIn=@n42) {baseReflectFraction=0.005, exponent=2.5}, Add=Multiply(A=@n30) {constB=0.1, desc=strength of gird in fresnel}) {materialFunction=/Engine/Functions/Engine_MaterialFunctions02/Math/MultiplyAdd}.Result
  ) {desc=colored fresnel}
  n44 = MaterialFunctionCall(Color=@n9) {materialFunction=&SplitComponents}
  n45 = ScalarParameter() {defaultValue=1, group=Wrap_Neon, parameterName=Neon_Stripes_TileRate_Weapon}
  n46 = MaterialFunctionCall(Color=@n9) {materialFunction=&SplitComponents}
  n47 = LinearInterpolate(
   A=ScalarParameter() {defaultValue=1, group=Wrap_Neon, parameterName=Neon_Stripes_TileRate_Vehicle}
   B=ScalarParameter() {defaultValue=1, group=Wrap_Neon, parameterName=Neon_Stripes_TileRate_VehicleSide}
   Alpha=@n46.G
  )
  n48 = MaterialFunctionCall!(
   For_Safe_Zones
   Normal_Map
   Caculated_Bounds=@n15
   Pre_Skinned_Normals=@n16
   Is Vehicle=@n44.R
   Is VehicleSide=@n44.R
   Is Mech=@n44.R
   Vehicle_Rotation=ScalarParameter() {group=Wrap_Neon, parameterName=Neon_Stripes_Rotation_Vehicle}
   UV_Rotation Bypass
   Mech_Rotation=ScalarParameter() {defaultValue=0.25, group=Wrap_Neon, parameterName=Neon_Stripes_Rotation_Mech}
   Uvs_Offset
   Offset-after rotation
   Uvs_TileRate=LinearInterpolate(
    A=LinearInterpolate(A=@n45, B=@n47, Alpha=@n46.R)
    B=ScalarParameter() {defaultValue=1, group=Wrap_Neon, parameterName=Neon_Stripes_TileRate_Mech}
    Alpha=@n46.B
   )
  ) {materialFunction=&/MF_Projection_Side_Normal}
  n49 = Multiply(A=@n48.UVs No Offsets/Tiling, B=ScalarParameter() {defaultValue=1, group=Wrap_Neon, parameterName=Neon_Warp_ATileRate})
  n50 = MaterialFunctionCall(Color=@n9) {materialFunction=&SplitComponents}
  n51 = LinearInterpolate(
   A=VectorParameter() {group=Wrap_Neon, parameterName=Neon_Warp_Speed_Weapon, defaultValue=<rgba 0 0 0 1>}.RGB
   B=LinearInterpolate(
    A=VectorParameter() {group=Wrap_Neon, parameterName=Neon_Warp_Speed_Vehicle, defaultValue=<rgba 0 0 0 1>}.RGB
    B=VectorParameter() {group=Wrap_Neon, parameterName=Neon_Warp_Speed_VehicleSide, defaultValue=<rgba 0 0 0 1>}.RGB
    Alpha=@n50.G
   )
   Alpha=@n50.R
  )
  n52 = VectorParameter() {group=Wrap_Neon, parameterName=Neon_Warp_Speed_Mech, defaultValue=<rgba 0 0 0 1>}
  n53 = ComponentMask(None=Multiply(A=Time(), B=LinearInterpolate(A=@n51, B=@n52.RGB, Alpha=@n50.B))) {g=true, r=true}
  n54 = TextureObjectParameter() {group=Wrap_Neon, parameterName=Neon_Warp__Text, texture=&T_BlacksRGB}
  n55 = Multiply(
   A=ScalarParameter() {defaultValue=0.5, group=Wrap_Neon, parameterName=Neon_Warp_Amount}
   B=TextureSample(UVs=Add(A=@n49, B=@n53), Tex=@n54) {samplerType=SAMPLERTYPE_LinearGrayscale}.R
  )
  n56 = Multiply(A=@n48.UVs No Offsets/Tiling, B=ScalarParameter() {defaultValue=1, group=Wrap_Neon, parameterName=Neon_Warp_BTileRate})
  n57 = ComponentMask(None=Multiply(A=Time(), B=VectorParameter() {group=Wrap_Neon, parameterName=Neon_Warp_B Speed, defaultValue=<rgba 0 0 0 1>}.RGB)) {g=true, r=true}
  n58 = Multiply(A=TextureSample(UVs=Add(A=@n56, B=@n57), Tex=@n54) {samplerType=SAMPLERTYPE_LinearGrayscale}.R, B=@n8)
  n59 = TextureSample(
   UVs=Add(A=@n48.UVs, B=Add(A=@n55, B=@n58))
   Tex=TextureObjectParameter() {group=Wrap_Neon, parameterName=Neon_Stripes__Text, texture=&T_BlacksRGB}
  ) {samplerType=SAMPLERTYPE_Masks}
  n60 = ScalarParameter() {group=Wrap_Neon, parameterName=Neon_Stripes_zGradent_Speed--------------}
  n61 = Multiply(A=Time(), B=@n60) {constB=0.5}
  n62 = Add(A=@n61, B=Multiply(A=@n7) {constB=2}) {constB=0.2}
  n63 = Add(A=Add(A=@n27, B=If(A=@n7, "A > B"=@n62, "A == B"=@n61, "A < B"=@n61) {constB=0.1}), B=@n3)
  n64 = Multiply(
   A=@n59.G
   B=MaterialFunctionCall(Hue Shift Percentage=Frac(None=@n63), Texture=Constant3Vector() {constant=<rgba 1 0 0 0>}) {materialFunction=&HueShift}.Result
  )
  n65 = ScalarParameter() {defaultValue=0.5, group=Wrap_Neon, parameterName=Neon_BS_Specular, sortPriority=1}
  n66 = ScalarParameter() {defaultValue=0.5, group=Wrap_Neon, parameterName=Neon_BS_Specular_High, sortPriority=1}
  n67 = ComponentMask(None=@n18.XYZ Texture) {r=true}
  n68 = ScalarParameter() {defaultValue=0.15, group=Wrap_Neon, parameterName=Neon_BS_Roughness, sortPriority=1}
  n69 = ScalarParameter() {defaultValue=1, group=Wrap_Neon, parameterName=Neon_BS_Roughness_High, sortPriority=1}
  n70 = ScalarParameter() {defaultValue=1, group=Wrap_Neon, parameterName=Neon_Stripes_Emissive_Amount}
  n71 = Constant() {r=0.5}
  n72 = Constant() {r=1}
  n73 = Constant() {r=0.25}
  n74 = SetMaterialAttributes!(
   MaterialAttributes=@n19.MaterialAttributes
   BaseColor=LinearInterpolate(A=Add(A=@n41, B=@n43), B=@n64, Alpha=@n59.G)
   Metallic=ScalarParameter() {group=Wrap_Neon, parameterName=Neon_BS_Metalic, sortPriority=1}
   Specular=LinearInterpolate(
    A=LinearInterpolate(A=@n65, B=@n66, Alpha=@n67)
    B=ScalarParameter() {defaultValue=1, group=Wrap_Neon, parameterName=Neon_BS_Specular_Stripes, sortPriority=1}
    Alpha=@n59.G
   )
   Roughness=LinearInterpolate(
    A=LinearInterpolate(A=@n68, B=@n69, Alpha=@n67)
    B=ScalarParameter() {defaultValue=0.3, group=Wrap_Neon, parameterName=Neon_BS_Roughness_Stripes, sortPriority=1}
    Alpha=@n59.G
   )
   EmissiveColor=Multiply(
    A=Multiply(A=@n64, B=@n70)
    B=MaterialFunctionCall(Value_Morning=@n71, Value_Day=@n72, Value_Evening=@n71, Value_Night=@n73) {materialFunction=&Scalar_Time_of_Day}.Result
   )
  ) {attributeSetTypes [
    #1
    #2
    #3
    #4
    #5
   ]}
  n75 = Saturate(None=Add(
    A=FunctionInput() {bUsePreviewValueAsDefault=true, inputName="Customization Mask (_CM B)", inputType=FunctionInput_Scalar, sortPriority=24}
    B=ScalarParameter() {group=Wrap_Neon, parameterName=Neon_Remove_Customization Mask, sortPriority=-10}
   ))
  FunctionOutput(None=StaticSwitchParameter(True=BlendMaterialAttributes(A=@n19.MaterialAttributes, B=@n74, Alpha=@n75) {vertexAttributeBlendType=UseA}, False=@n19.MaterialAttributes) {group=Customization_Wrap, parameterName=Wrap_Neon}) {outputName=Result_1}
  Multiply(A=@n64)
  Multiply(A=@n21.Raw_Data)
  Multiply(A=Time(), B=@n60)
  ScalarParameter() {defaultValue=0.5, parameterName=Neon_AudioPanSpeed}
 ]}
