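# Material function for the NeonBands weapon wrap, painting scrolling, glowing color bands on weapons, vehicles and mechs.
# Per-type object-space sphere gradients, warped by normals and gradient noise, drive a curve-atlas color lookup with view-angle falloff.
# Event and emote audio modulate hue and emissive on selected colors, while three timestamp parameters trigger fading flashes through Grad_Mask channels.
# Results blend over input attributes by customization mask behind the Wrap_NeonBands static switch.
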MaterialFunction /Game/Weapons/WeaponSkins/Wraps/MaterialFunctions/MF_Wrap_NeonBands
graph {nodes [
  Clamp()
  n1 = CollectionParameter() {collection=&FortniteMaterialParameters, parameterName=EventAudio_Spectrum}
  n2 = MaterialFunctionCall(Float4=@n1) {materialFunction=&BreakOutFloat4Components}
  n3 = Time()
  n4 = Add(A=Sine(None=Multiply(A=@n3) {constB=0.5}))
  n5 = Power(Base=Multiply(A=LinearInterpolate(A=@n2.G, B=@n4) {constAlpha=0.7}) {constB=1.25}) {constExponent=5}
  Clamp(None=@n5)
  n6 = VectorParameter() {group=Wrap_NeonBands, parameterName=NeonBnds_Paining_Weapon, defaultValue=<rgba 0 0 0 1>}
  n7 = VectorParameter() {group=Wrap_NeonBands, parameterName=NeonBnds_Paining_vehicle, defaultValue=<rgba 0 0 0 1>}
  n8 = FunctionInput() {inputName=isVehicle, inputType=FunctionInput_Scalar, sortPriority=5}
  n9 = FunctionInput() {bUsePreviewValueAsDefault=true, inputName=isMech, inputType=FunctionInput_Scalar, sortPriority=4}
  n10 = Multiply(
   A=Time()
   B=LinearInterpolate(
    A=LinearInterpolate(A=@n6.RGB, B=@n7.RGB, Alpha=@n8)
    B=VectorParameter() {group=Wrap_NeonBands, parameterName=NeonBnds_Paining_Mech, defaultValue=<rgba 0 0 0 1>}.RGB
    Alpha=@n9
   )
  )
  Frac(None=@n10)
  FunctionInput() {inputName=isVehicle_Side, inputType=FunctionInput_Scalar, sortPriority=5}
  n11 = PixelNormalWS()
  n12 = CameraVectorWS()
  n13 = Power(
   Base=Clamp(None=DotProduct(A=@n11, B=@n12))
   Exp=ScalarParameter() {defaultValue=2, group=Wrap_NeonBands, parameterName=NeonBnds_ViewDepedantFalloff}
  )
  n14 = ScalarParameter() {group=Wrap_NeonBands, parameterName="NeonBnds_ViewDepedantFalloff 0=active 1=off"}
  n15 = Frac(None=Multiply(A=Time(), B=ScalarParameter() {defaultValue=0.25, group=Wrap_NeonBands, parameterName=NeonBnds_Hueshift_Speed, sortPriority=200}) {constB=0.25})
  n16 = Round(None=Multiply(A=@n2.R) {constB=1000})
  n17 = CollectionParameter() {collection=&FortniteMaterialParameters, parameterName=Emote_AudioLow}
  n18 = Clamp(None=Round(None=Multiply(A=@n17) {constB=200}))
  n19 = ScalarParameter() {group=Wrap_NeonBands, parameterName=NeonBnds_Disable_audio, sortPriority=-20}
  n20 = LinearInterpolate(A=LinearInterpolate(A=Clamp(None=@n16), B=@n18, Alpha=@n18), Alpha=@n19) {constB=0}
  n21 = VertexInterpolator(VS=PreSkinnedPosition())
  n22 = PreSkinnedLocalBounds()
  n23 = FunctionInput(Preview=Divide(A=Subtract(A=@n21.PS, B=@n22.Min), B=Subtract(A=@n22.Max, B=@n22.Max))) {bUsePreviewValueAsDefault=true, inputName=Bounds}
  n24 = VectorParameter() {group=Wrap_NeonBands, parameterName=NeonBnds_ObjectSpaceGradStart_Weapon, defaultValue=<rgba 0 0 0 1>}
  n25 = ScalarParameter() {defaultValue=1, group=Wrap_NeonBands, parameterName=NeonBnds_ObjectSpaceGradRadius_Weapon}
  n26 = Constant()
  n27 = VectorParameter() {group=Wrap_NeonBands, parameterName=NeonBnds_ObjectSpaceGradStart_Vehicle, defaultValue=<rgba 0 0 0 1>}
  n28 = ScalarParameter() {defaultValue=1, group=Wrap_NeonBands, parameterName=NeonBnds_ObjectSpaceGradRadius_Vehicle}
  n29 = Constant()
  n30 = LinearInterpolate(
   A=LinearInterpolate(A=SphereMask(A=@n23, B=@n24.RGB, Radius=@n25, Hardness=@n26), B=SphereMask(A=@n23, B=@n27.RGB, Radius=@n28, Hardness=@n29), Alpha=@n8)
   B=SphereMask(
    A=@n23
    B=VectorParameter() {group=Wrap_NeonBands, parameterName=NeonBnds_ObjectSpaceGradStart_Mech, defaultValue=<rgba 0 0 0 1>}.RGB
    Radius=ScalarParameter() {defaultValue=1, group=Wrap_NeonBands, parameterName=NeonBnds_ObjectSpaceGradRadius_Mech}
    Hardness=Constant()
   )
   Alpha=@n9
  )
  n31 = GetMaterialAttributes(None=FunctionInput() {bUsePreviewValueAsDefault=true, inputName=Material, inputType=FunctionInput_MaterialAttributes, sortPriority=-1}) {attributeGetTypes [#1]}
  n32 = Multiply(
   A=ComponentMask(None=@n31.Normal) {b=true}
   B=ScalarParameter() {defaultValue=0.2, group=Wrap_NeonBands, parameterName=NeonBnds_NormalDistortionWeight}
  )
  n33 = Multiply(A=Time(), B=ScalarParameter() {defaultValue=0.1, group=Wrap_NeonBands, parameterName=NeonBnds_speed})
  n34 = Add(
   A=Multiply(A=Add(A=@n30, B=@n32), B=ScalarParameter() {defaultValue=1, group=Wrap_NeonBands, parameterName=NeonBnds_StretchGradient})
   B=@n33
  )
  n35 = Multiply(A=Add(A=@n23, B=@n10), B=ScalarParameter() {defaultValue=75, group=Wrap_NeonBands, parameterName=NeonBnds_Noise_tile})
  n36 = Multiply(
   A=ComponentMask(None=Noise!(World Position=@n35, FilterWidth) {bTurbulence=false, levels=1, noiseFunction=NOISEFUNCTION_GradientALU, outputMin=0, scale=0.04}) {g=true}
   B=ScalarParameter() {defaultValue=1, group=Wrap_NeonBands, parameterName=NeonBnds_NoiseWarpIntensity}
  )
  n37 = Frac(None=Multiply(A=Add(A=@n34, B=@n36), B=ScalarParameter() {defaultValue=1.6, group=Wrap_NeonBands, parameterName=NeonBnds_Repetition}))
  n38 = CurveAtlasRowParameter(CurveTime=@n37) {
   atlas=/Game/Weapons/WeaponSkins/Wraps/Textures/CruveAtlas_Wraps_01
   curve=/Game/Weapons/WeaponSkins/Wraps/Textures/NeonBands/CC_NeonBands
   defaultValue=16
   group=Wrap_NeonBands
   parameterName=NeonBnds_ColorCurve
  }
  n39 = MaterialFunctionCall!(
   Index=ScalarParameter() {defaultValue=16, group=Wrap_NeonBands, parameterName="NeonBnds_ColorCurve  Index", sortPriority=200}
   CCinput=@n38
   CCinput_Alpha
   UV=@n37
  ) {materialFunction=&/MF_Global_CC_Tex}
  n40 = VectorParameter() {group=Wrap_NeonBands, parameterName=NeonBands_Color_Selection_Audio_A, defaultValue=<rgba 0.9333 0.8039 0.0431 1>}
  n41 = VectorParameter() {group=Wrap_NeonBands, parameterName=NeonBands_Color_Selection_Audio_B, defaultValue=<rgba 0.1804 0.8196 1 1>}
  n42 = Frac(None=Multiply(A=@n33) {constB=4})
  n43 = LinearInterpolate(A=@n40.RGB, B=@n41.RGB, Alpha=Round(None=Saturate(None=@n42)))
  n44 = SphereMask(
   A=Saturate(None=DotProduct(A=@n39.3V, B=@n43))
   B=Constant() {r=1}
   Radius=ScalarParameter() {defaultValue=0.05, group=Wrap_NeonBands, parameterName=NeonBnds_SphereRad, sortPriority=200}
   Hardness=Constant() {r=1}
  )
  n45 = MaterialFunctionCall(Hue Shift Percentage=Multiply(A=Multiply(A=@n15, B=@n20), B=@n44), Texture=@n39.3V) {materialFunction=&HueShift}
  n46 = Multiply(A=Saturate(None=Add(A=@n13, B=@n14)), B=@n45.Result)
  n47 = ScalarParameter() {defaultValue=0.75, group=Wrap_NeonBands, parameterName=NeonBnds_Emissive_Strenght, sortPriority=-9}
  n48 = ScalarParameter() {defaultValue=0.3, group=Wrap_NeonBands, parameterName=NeonBnds_Emissive, sortPriority=200}
  n49 = Clamp(None=LinearInterpolate(A=@n5, B=@n17, Alpha=@n18))
  n50 = LinearInterpolate(
   A=ScalarParameter() {defaultValue=0.3, group=Wrap_NeonBands, parameterName=NeonBnds_EmissiveAccent, sortPriority=200}
   B=ScalarParameter() {defaultValue=3, group=Wrap_NeonBands, parameterName=NeonBnds_Yellow Boost, sortPriority=200}
   Alpha=Multiply(A=@n44, B=LinearInterpolate(A=@n49, Alpha=@n19) {constB=0})
  )
  n51 = TextureSampleParameter2D(UVs=@n37) {
   group=Wrap_NeonBands
   parameterName=Grad_Mask
   samplerType=SAMPLERTYPE_Masks
   texture=/Game/Weapons/WeaponSkins/Wraps/Textures/NeonBands/T_NeonBands_Grad
  }
  n52 = ScalarParameter() {group=Wrap_NeonBands, parameterName=NeonBnds_Reactive_Emissive_max, sortPriority=10}
  n53 = Time()
  n54 = ScalarParameter() {defaultValue=-9999999, group=Wrap_NeonBands, parameterName=NeonBnds_MainA_TimeFronBP}
  n55 = Saturate(None=Divide(A=Subtract(A=@n53, B=@n54), B=ScalarParameter() {defaultValue=0.5, group=Wrap_NeonBands, parameterName=NeonBnds_Duration_A}))
  n56 = Max(
   A=Multiply(A=Multiply(A=@n46, B=@n47) {constB=0.5}, B=LinearInterpolate(A=@n48, B=@n50, Alpha=@n20))
   B=Multiply(A=Multiply(A=@n51.R, B=@n46), B=LinearInterpolate(A=@n52, Alpha=@n55) {constA=5, constB=0})
  )
  n57 = Divide(
   A=Subtract(A=Time(), B=ScalarParameter() {defaultValue=-9999999, group=Wrap_NeonBands, parameterName=NeonBnds_MainB_TimeFronBP})
   B=ScalarParameter() {defaultValue=0.5, group=Wrap_NeonBands, parameterName=NeonBnds_Duration_B}
  )
  n58 = Multiply(A=Multiply(A=@n46, B=@n51.G), B=LinearInterpolate(A=@n52, Alpha=Saturate(None=@n57)) {constA=5, constB=0})
  n59 = Multiply(A=@n46, B=@n51.B)
  n60 = Subtract(A=Time(), B=ScalarParameter() {defaultValue=-9999999, group=Wrap_NeonBands, parameterName=NeonBnds_MainC_TimeFronBP})
  n61 = ScalarParameter() {defaultValue=0.5, group=Wrap_NeonBands, parameterName=NeonBnds_Duration_C}
  n62 = LinearInterpolate(A=@n52, Alpha=Saturate(None=Divide(A=@n60, B=@n61))) {constA=5, constB=0}
  n63 = SetMaterialAttributes!(
   MaterialAttributes=@n31.MaterialAttributes
   ShadingModel
   BaseColor=Multiply(A=@n46, B=ScalarParameter() {defaultValue=1, group=Wrap_NeonBands, parameterName=NeonBnds_BaseColorBrightness, sortPriority=-10})
   Metallic=ScalarParameter() {defaultValue=1, group=Wrap_NeonBands, parameterName=NeonBnds_Metalness, sortPriority=-10}
   Specular=ScalarParameter() {defaultValue=0.5, group=Wrap_NeonBands, parameterName=NeonBnds_Specular, sortPriority=-10}
   Roughness=ScalarParameter() {defaultValue=0.5, group=Wrap_NeonBands, parameterName=NeonBnds_Roughness, sortPriority=-10}
   Normal=MaterialFunctionCall(Normal=@n31.Normal, Flatness=ScalarParameter() {defaultValue=0.5, group=Wrap_NeonBands, parameterName=NeonBnds_DanceLines_NormalFlattness}) {materialFunction=&FlattenNormal}.Result
   EmissiveColor=Max(A=Max(A=@n56, B=@n58), B=Multiply(A=@n59, B=@n62))
  ) {attributeSetTypes [
    #2
    #3
    #4
    #5
    #6
    #1
    #7
   ]}
  n64 = FunctionInput() {bUsePreviewValueAsDefault=true, inputName="Customization Mask (_CM B)", inputType=FunctionInput_Scalar, sortPriority=24}
  n65 = ScalarParameter() {group=Wrap_NeonBands, parameterName=NeonBnds_Remove_Customization Mask, sortPriority=-10}
  n66 = BlendMaterialAttributes(A=@n31.MaterialAttributes, B=@n63, Alpha=Saturate(None=Add(A=@n64, B=@n65)))
  FunctionOutput(None=StaticSwitchParameter(True=@n66, False=@n31.MaterialAttributes) {group=Customization_Wrap, parameterName=Wrap_NeonBands})
  n67 = CollectionParameter() {collection=&FortniteMaterialParameters, parameterName=EventAudio_Spectrum}
  n68 = MaterialFunctionCall(Float4=@n67) {materialFunction=&BreakOutFloat4Components}
  n69 = ScalarParameter() {defaultValue=10, group=Wrap_NeonBands, parameterName=NeonBnds_Audio_Event_mask_Multplyer}
  n70 = Saturate(None=Round(None=Multiply(A=@n68.R, B=@n69) {constB=10}))
  n71 = CollectionParameter() {collection=&FortniteMaterialParameters, parameterName=Emote_AudioLow}
  n72 = Round(None=Multiply(A=@n71, B=ScalarParameter() {defaultValue=200, group=Wrap_NeonBands, parameterName=NeonBnds_Audio_Emote_Low_Multplyer}) {constB=200})
  LinearInterpolate(Alpha=Saturate(None=LinearInterpolate(A=@n70, B=@n72, Alpha=@n72)))
  MaterialFunctionCall!(In=@n45.Result, Contrast=@n20) {materialFunction=/Engine/Functions/Engine_MaterialFunctions01/ImageAdjustment/CheapContrast_RGB}
  Multiply(
   A=SphereMask(
    A=MaterialFunctionCall(Float3=@n23) {materialFunction=&BreakOutFloat3Components}.G
    B=ScalarParameter() {defaultValue=150, desc=9.176, group=Wrap_NeonBands, parameterName=NeonBnds_YAxisGradPosition}
    Radius=ScalarParameter() {defaultValue=50, desc=50, group=Wrap_NeonBands, parameterName=NeonBnds_YAxisGradRadius}
    Hardness=Constant()
   )
   B=ScalarParameter() {defaultValue=1, group=Wrap_NeonBands, parameterName=NeonBnds_YAxisGradIntensity}
  )
  n73 = Power(
   Base=Multiply(A=@n68.R, B=ScalarParameter() {defaultValue=1.25, group=Wrap_NeonBands, parameterName=NeonBnds_Audio_Event_Multplyer}) {constB=1.25}
   Exp=ScalarParameter() {defaultValue=5, group=Wrap_NeonBands, parameterName=NeonBnds_Audio_Event_Power}
  ) {constExponent=5}
  Saturate(None=LinearInterpolate(A=Saturate(None=@n73), B=@n71, Alpha=@n72))
  ShadingModel() {shadingModel=MSM_ClearCoat}
  StaticBoolParameter() {parameterName=Remove Emote Audio Priority}
  StaticSwitchParameter(True=@n66, False=Constant()) {parameterName=sdfgsry65434}
 ]}
