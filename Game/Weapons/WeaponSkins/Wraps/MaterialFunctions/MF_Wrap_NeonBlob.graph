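# Weapon wrap material function that, behind the Wrap_NeonBlob static switch, turns masked areas into animated glowing neon blobs with a separate accent material.
# Projected UVs, tuned per item type, drive three panning, pulsing bubble layers combined with noise and a box mask.
# Curve atlases shape the result into a blob mask and time-cycling colors for emissive, scaled by time of day and brightened by curvature.
# The mask also drives roughness, specular and procedurally generated normals.
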MaterialFunction /Game/Weapons/WeaponSkins/Wraps/MaterialFunctions/MF_Wrap_NeonBlob
graph {nodes [
  n1 = TextureSample() {samplerType=SAMPLERTYPE_Masks, texture=/Game/Weapons/FORT_Rifles/Textures/T_Scar_CM}
  n2 = FunctionInput(Preview=@n1.R) {bUsePreviewValueAsDefault=true, inputName=Mask Primary, inputType=FunctionInput_Scalar, sortPriority=24, previewValue=<1 0 0 1>}
  n3 = FunctionInput(Preview=@n1.B) {bUsePreviewValueAsDefault=true, inputName=Mask Secondary, inputType=FunctionInput_Scalar, sortPriority=24, previewValue=<1 0 0 1>}
  n4 = BlendMaterialAttributes(A=NamedRerouteUsage(), B=NamedRerouteUsage(), Alpha=Saturate(None=Add(A=@n2, B=@n3))) {vertexAttributeBlendType=UseA}
  n5 = NamedRerouteUsage()
  n6 = NamedRerouteDeclaration(None=@n3) {name=Msk_Accent, nodeColor=<rgba 0 1 0.7656 1>}
  n7 = TextureSample() {samplerType=SAMPLERTYPE_Normal, texture=/Game/Weapons/FORT_Rifles/T_SCAR_N}
  n8 = Constant3Vector() {constant=<rgba 1 0 0 0>}
  n9 = GetMaterialAttributes(None=FunctionInput(Preview=SetMaterialAttributes!(MaterialAttributes, Normal=@n7.RGB, BaseColor=@n8) {attributeSetTypes [#1, #2]}) {inputName=Material, inputType=FunctionInput_MaterialAttributes}) {attributeGetTypes [
    #2
    #3
    #1
    #4
   ]}
  FunctionOutput(None=StaticSwitchParameter(
    True=BlendMaterialAttributes(A=@n4, B=@n5, Alpha=@n6) {vertexAttributeBlendType=UseA}
    False=NamedRerouteDeclaration(None=@n9.MaterialAttributes) {name=Root_Material, nodeColor=<rgba 0.7812 1 0 1>}
   ) {group=Customization_Wrap, parameterName=Wrap_NeonBlob})
  n10 = VectorParameter() {group=Wrap_NeonBlob, parameterName=NeonBL_Color_Bck, defaultValue=<rgba 0.005 0.005 0.005 0>}
  n11 = NamedRerouteUsage()
  n12 = Add(A=NamedRerouteUsage()) {constB=-0.5}
  n13 = Power(Base=NamedRerouteUsage())
  n14 = Add(A=NamedRerouteUsage(), B=ComponentMask(None=NamedRerouteUsage()) {g=true, r=true}) {constB=-0.5}
  n15 = Subtract(A=Abs(None=Multiply(A=@n14) {constB=2}), B=ComponentMask(None=NamedRerouteUsage()) {g=true, r=true})
  n16 = Max(A=@n15) {constB=0}
  n17 = MaterialFunctionCall(Float2=@n15) {materialFunction=&BreakOutFloat2Components}
  n18 = Max(A=@n17.R, B=@n17.R) {constB=0}
  n19 = Divide(
   A=Add(
    A=MaterialFunctionCall(Vector 2=@n16) {materialFunction=/Engine/Functions/Engine_MaterialFunctions02/Utility/VectorLength}.V2 Length
    B=Min(A=@n18) {constB=0}
   )
   B=ScalarParameter() {defaultValue=0.15, group=Wrap_NeonBlob_Uvs, parameterName=NeonBL_Box_Blend_Falloff}
  ) {constB=0.1}
  n20 = Saturate(None=Power(Base=OneMinus(None=@n19)))
  n21 = Multiply(A=@n20, B=@n13)
  n22 = ComponentMask(None=NamedRerouteUsage()) {r=true}
  n23 = CurveAtlasRowParameter(CurveTime=@n22) {
   atlas=/Game/Weapons/WeaponSkins/Wraps/Textures/CruveAtlas_Wraps_01
   curve=/Game/Weapons/WeaponSkins/Wraps/Textures/NeonBlob/CC_MskA
   defaultValue=105
   group=Wrap_NeonBlob_CC
   parameterName=NeonBL_CC_MskA
  }
  n24 = ComponentMask(None=MaterialFunctionCall!(
    Index=ScalarParameter() {defaultValue=105, group=Wrap_NeonBlob_CC, parameterName="NeonBL_CC_MskA  Index"}
    CCinput=@n23
    CCinput_Alpha
    UV=@n22
   ) {materialFunction=&/MF_Global_CC_Tex}.3V) {r=true}
  n25 = Multiply(
   A=Multiply(A=NamedRerouteUsage(), B=ScalarParameter() {defaultValue=1, group=Wrap_NeonBlob_Bubbles, parameterName=NeonBL_BubbleA_Scale})
   B=NamedRerouteUsage()
  )
  n26 = ComponentMask(None=VectorParameter() {group=Wrap_NeonBlob_Bubbles, parameterName=NeonBL_BubbleA_Speed, defaultValue=<rgba -0.005 0.01 0 1>}.RGB) {g=true, r=true}
  n27 = TextureSample(UVs=Add(A=Panner(Coordinate=@n25, Speed=@n26), B=NamedRerouteUsage()), Tex=NamedRerouteUsage()) {samplerType=SAMPLERTYPE_Masks}
  n28 = Add(A=Multiply(A=Time(), B=ScalarParameter() {defaultValue=0.15, group=Wrap_NeonBlob_Bubbles, parameterName=NeonBL_BubbleA_Msk_Speed}), B=@n27.R)
  n29 = LinearInterpolate(
   A=ScalarParameter() {defaultValue=0.3, group=Wrap_NeonBlob_Bubbles, parameterName=NeonBL_Bubble_A_Size_Max}
   B=ScalarParameter() {defaultValue=0.5, group=Wrap_NeonBlob_Bubbles, parameterName=NeonBL_Bubble_A_Size_Min}
   Alpha=ConstantBiasScale(None=Sine(None=@n28))
  ) {constA=0.4, constB=0.49}
  n30 = ScalarParameter() {defaultValue=2, group=Wrap_NeonBlob_Bubbles, parameterName=NeonBL_Bubble_Fallof}
  n31 = Power(
   Base=MaterialFunctionCall(Transition Phase=Multiply(A=@n24, B=@n29), Height Texture=@n27.A) {materialFunction=&HeightLerp}.Lerp Alpha No Contrast
   Exp=@n30
  )
  n32 = Time()
  n33 = ScalarParameter() {defaultValue=0.1, group=Wrap_NeonBlob_Bubbles, parameterName=NeonBL_BubbleB_Msk_Speed}
  n34 = Multiply(
   A=Multiply(A=NamedRerouteUsage(), B=ScalarParameter() {defaultValue=1.5, group=Wrap_NeonBlob_Bubbles, parameterName=NeonBL_BubbleB_Scale})
   B=NamedRerouteUsage()
  )
  n35 = ComponentMask(None=VectorParameter() {group=Wrap_NeonBlob_Bubbles, parameterName=NeonBL_BubbleB_Speed, defaultValue=<rgba 0.025 0.005 0 1>}.RGB) {g=true, r=true}
  n36 = TextureSample(UVs=Add(A=Panner(Coordinate=@n34, Speed=@n35), B=NamedRerouteUsage()), Tex=NamedRerouteUsage()) {samplerType=SAMPLERTYPE_Masks}
  n37 = Sine(None=Add(A=Multiply(A=@n32, B=@n33), B=@n36.R))
  n38 = Multiply(
   A=@n24
   B=LinearInterpolate(
    A=ScalarParameter() {defaultValue=0.35, group=Wrap_NeonBlob_Bubbles, parameterName=NeonBL_Bubble_B_Size_Max}
    B=ScalarParameter() {defaultValue=0.5, group=Wrap_NeonBlob_Bubbles, parameterName=NeonBL_Bubble_B_Size_Min}
    Alpha=ConstantBiasScale(None=@n37)
   ) {constA=0.4, constB=0.49}
  )
  n39 = Time()
  n40 = ScalarParameter() {defaultValue=0.15, group=Wrap_NeonBlob_Bubbles, parameterName=NeonBL_BubbleC_Msk_Speed}
  n41 = Multiply(
   A=Multiply(A=NamedRerouteUsage(), B=ScalarParameter() {defaultValue=1.75, group=Wrap_NeonBlob_Bubbles, parameterName=NeonBL_BubbleC_Scale})
   B=NamedRerouteUsage()
  )
  n42 = ComponentMask(None=VectorParameter() {group=Wrap_NeonBlob_Bubbles, parameterName=NeonBL_BubbleC_Speed, defaultValue=<rgba -0.015 -0.015 0 1>}.RGB) {g=true, r=true}
  n43 = TextureSample(UVs=Add(A=Panner(Coordinate=@n41, Speed=@n42), B=NamedRerouteUsage()), Tex=NamedRerouteUsage()) {samplerType=SAMPLERTYPE_Masks}
  n44 = Sine(None=Add(A=Multiply(A=@n39, B=@n40), B=@n43.R))
  n45 = Multiply(
   A=@n24
   B=LinearInterpolate(
    A=ScalarParameter() {defaultValue=0.25, group=Wrap_NeonBlob_Bubbles, parameterName=NeonBL_Bubble_C_Size_Max}
    B=ScalarParameter() {defaultValue=0.5, group=Wrap_NeonBlob_Bubbles, parameterName=NeonBL_Bubble_C_Size_Min}
    Alpha=ConstantBiasScale(None=@n44)
   ) {constA=0.4, constB=0.49}
  )
  n46 = Add(
   A=Power(
    Base=MaterialFunctionCall(Transition Phase=@n38, Height Texture=@n36.A) {materialFunction=&HeightLerp}.Lerp Alpha No Contrast
    Exp=@n30
   )
   B=Power(
    Base=MaterialFunctionCall(Transition Phase=@n45, Height Texture=@n43.A) {materialFunction=&HeightLerp}.Lerp Alpha No Contrast
    Exp=@n30
   )
  )
  n47 = Multiply(A=NamedRerouteUsage(), B=Add(A=Multiply(A=@n21) {constB=0.25}, B=Add(A=@n31, B=@n46)))
  n48 = Add(A=Multiply(A=Multiply(A=@n11, B=@n12), B=NamedRerouteUsage()), B=@n47)
  n49 = CurveAtlasRowParameter(CurveTime=@n48) {
   atlas=&CruveAtlas_Wraps_01
   curve=&CC_MskA
   defaultValue=105
   group=Wrap_NeonBlob_CC
   parameterName=NeonBL_CC_MskA
  }
  MaterialExpressionMaterialFunctionCall_27 = MaterialFunctionCall(
   Index=ScalarParameter() {defaultValue=105, group=Wrap_NeonBlob_CC, parameterName="NeonBL_CC_MskA  Index"}
   CCinput=@n49
   CCinput_Alpha=@n49
   UV=@n48
  ) {materialFunction=&MF_Global_CC_Tex}
  n50 = Convert(None=@MaterialExpressionMaterialFunctionCall_27.3V) {
   nodeName=Break Float3
   convertInputs [{
     type=Vector3
     expressionInput {
      expression=&self:MaterialExpressionMaterialFunctionCall_27
      inputName=None
      mask=0
      maskA=0
      maskB=0
      maskG=0
      maskR=0
      outputIndex=0
     }
    }]
   convertOutputs [{type=Scalar}, {type=Scalar}, {type=Scalar}]
  }
  n51 = Multiply(A=@n50, B=ScalarParameter() {defaultValue=3, group=Wrap_NeonBlob_CC, parameterName=NeonBL_CC__Edge}) {constB=5}
  n52 = ScalarParameter() {defaultValue=0.5, group=Wrap_NeonBlob_CC, parameterName=NeonBL_CC__Edge_Strength}
  n53 = ScalarParameter() {defaultValue=0.25, group=Wrap_NeonBlob_CC, parameterName=NeonBL_CC__TileRate}
  n54 = ComponentMask(None=NamedRerouteUsage()) {r=true}
  n55 = Add(A=@n54, B=NamedRerouteUsage())
  n56 = Multiply(
   A=Saturate(None=Add(A=@n51, B=@n52) {constB=0.5})
   B=Add(A=Multiply(A=@MaterialExpressionMaterialFunctionCall_27.Alpha, B=@n53), B=Multiply(A=@n54, B=@n55))
  )
  n57 = Multiply(A=Time(), B=ScalarParameter() {defaultValue=0.05, group=Wrap_NeonBlob_CC, parameterName=NeonBL_CC_Msk_Speed})
  n58 = Frac(None=Add(A=Add(A=@n56, B=@n57), B=ScalarParameter() {group=Wrap_NeonBlob_CC, parameterName=NeonBL_CC_Shift}))
  n59 = CurveAtlasRowParameter(CurveTime=@n58) {
   atlas=&CruveAtlas_Wraps_01
   curve=/Game/Weapons/WeaponSkins/Wraps/Textures/NeonBlob/CC_NB_Colors
   defaultValue=106
   group=Wrap_NeonBlob_CC
   parameterName=NeonBL_CC_Blobs
  }
  n60 = MaterialFunctionCall(Index=ScalarParameter() {defaultValue=106, group=Wrap_NeonBlob_CC, parameterName="NeonBL_CC_Blobs  Index"}, CCinput=@n59, UV=@n58) {materialFunction=&MF_Global_CC_Tex}
  LinearInterpolate(A=@n10.RGB, B=@n60.3V, Alpha=Saturate(None=Add(A=@n10.A, B=@n50)))
  n61 = Multiply(A=@n50, B=@n60.3V)
  n62 = ScalarParameter() {defaultValue=5, group=Wrap_NeonBlob, parameterName=NeonGL_Curavture_Brightness}
  n63 = Clamp(None=Add(A=NamedRerouteUsage()) {constB=-0.5}) {minDefault=-0.1}
  n64 = LinearInterpolate(A=@n61, B=Multiply(A=@n61, B=@n62), Alpha=@n63)
  n65 = VectorParameter() {group=Wrap_BadBear_Subsurface, parameterName=BadBear_InnerScatteringColor, defaultValue=<rgba 0.0877 0.0586 0.2031 1>}
  n66 = CollectionParameter() {
   collection=&FortniteMaterialParameters
   parameterName=SunlightDirectionalVector
  }
  n67 = ComponentMask(None=@n66) {b=true, g=true, r=true}
  n68 = PixelNormalWS()
  n69 = Multiply(A=DotProduct(A=@n67, B=@n68)) {constB=2}
  n70 = Normalize(VectorInput=Subtract(A=Multiply(A=@n69, B=@n68), B=@n67))
  n71 = Multiply(A=MaterialFunctionCall() {materialFunction=&CameraDirectionVector}.Result) {constB=-1}
  n72 = Power(
   Base=Max(A=DotProduct(A=@n70, B=@n71)) {constB=0}
   Exp=ScalarParameter() {defaultValue=5, group=Wrap_NeonBlob, parameterName=NeonBL_Light_Fake}
  )
  n73 = ScalarParameter() {defaultValue=0.4, group=Wrap_NeonBlob, parameterName=NeonBL_Light_Fake_Brightness}
  LinearInterpolate(A=@n64, B=@n65.RGB, Alpha=Saturate(None=Multiply(A=@n72, B=@n73) {constB=0.4}))
  n74 = Constant() {r=1}
  n75 = NamedRerouteUsage()
  n76 = MaterialFunctionCall(Float4=NamedRerouteUsage()) {materialFunction=&BreakOutFloat4Components}
  n77 = LinearInterpolate(
   A=@n74
   B=MaterialFunctionCall(Color=Abs(None=@n75)) {materialFunction=&SplitComponents}.G
   Alpha=@n76.G
  )
  n78 = ScalarParameter() {defaultValue=0.5, group=Wrap_NeonBlob, parameterName=NeonBLBlend_Contrast_VehSide}
  MaterialFunctionCall(In=LinearInterpolate(A=LinearInterpolate(A=@n74, B=@n77, Alpha=@n76.R), B=@n74, Alpha=@n76.B), Contrast=@n78) {materialFunction=&CheapContrast}
  MaterialFunctionCall(UVs=NamedRerouteUsage(), Rotation Center=Constant() {r=0.5}, "Rotation Angle (0-1)"=NamedRerouteUsage()) {materialFunction=&CustomRotator}
  Multiply(A=@n22)
  n79 = Multiply(A=@n20, B=@n13)
  Multiply(A=@n79, B=Add(A=@n79, B=@n13))
  Multiply(A=@n64, B=ScalarParameter() {defaultValue=5, group=Wrap_NeonBlob, parameterName=NeonBL_Light_Brightness}) {constB=5}
  Multiply(A=MaterialFunctionCall(Value_Morning=Constant() {r=0.5}, Value_Day=Constant() {r=1}, Value_Evening=Constant() {r=0.5}, Value_Night=Constant() {r=0.25}) {materialFunction=&Scalar_Time_of_Day}.Result)
  n80 = VectorParameter() {group=Wrap_NeonBlob_Accent, parameterName=NeonBL_Accent_BaseColor, defaultValue=<rgba 0.009 0.009 0.009 1>}
  n81 = SetMaterialAttributes!(
   MaterialAttributes=NamedRerouteUsage()
   BaseColor=@n80.RGB
   Roughness=ScalarParameter() {defaultValue=0.6, group=Wrap_NeonBlob_Accent, parameterName=NeonBL_Accent_Roughness}
   Specular=ScalarParameter() {defaultValue=0.5, group=Wrap_NeonBlob_Accent, parameterName=NeonBL_Accent_Spec}
   Metallic=ScalarParameter() {group=Wrap_NeonBlob_Accent, parameterName=NeonBL_Accent_Metalic}
   Normal
   EmissiveColor
  ) {attributeSetTypes [
    #2
    #3
    #5
    #4
    #1
    #6
   ]}
  NamedRerouteDeclaration(None=@n81) {name=Mat_Accent, nodeColor=<rgba 0 0.3906 1 1>}
  n82 = FunctionInput(Preview=@n1.G) {bUsePreviewValueAsDefault=true, inputName=Mask Crevise, inputType=FunctionInput_Scalar, sortPriority=24, previewValue=<1 0 0 1>}
  NamedRerouteDeclaration(None=@n82) {name=Msk_Curvature, nodeColor=<rgba 0 1 0.7422 1>}
  n83 = CurveAtlasRowParameter(CurveTime=@n47) {
   atlas=&CruveAtlas_Wraps_01
   curve=&CC_MskA
   defaultValue=105
   group=Wrap_NeonBlob_CC
   parameterName=NeonBL_CC_MskA
  }
  n84 = MaterialFunctionCall!(Value=MaterialFunctionCall(Index=ScalarParameter() {defaultValue=105, group=Wrap_NeonBlob_CC, parameterName="NeonBL_CC_MskA  Index"}, CCinput_Alpha=@n83.A, UV=@n47) {materialFunction=&MF_Global_CC_Tex}.Alpha) {materialFunction=/Engine/Functions/Engine_MaterialFunctions03/Procedurals/PreparePerturbNormalHQ}
  n85 = Multiply(
   A=Transform(None=MaterialFunctionCall!(
     Bump one pixel right=@n84.Value one pixel right
     Bump center=@n84.Value one pixel right
     Bump one pixel down=@n84.Value one pixel right
     WS Normal
    ) {materialFunction=/Engine/Functions/Engine_MaterialFunctions03/Procedurals/PerturbNormalHQ}.WS Normal) {transformSourceType=TRANSFORMSOURCE_World, transformType=TRANSFORM_Tangent}
   B=Constant3Vector() {constant=<rgba 1 1 0 0>}
  )
  NamedRerouteDeclaration(None=Multiply(A=@n85, B=ScalarParameter() {defaultValue=1, group=Wrap_NeonBlob, parameterName=NeonBL_Normal_Strength})) {name=Generated_Normals, nodeColor=<rgba 0.6406 0 1 1>}
  n86 = MaterialFunctionCall(Float4=NamedRerouteUsage()) {materialFunction=&BreakOutFloat4Components}
  n87 = MaterialFunctionCall!(
   For_Safe_Zones
   Normal_Map
   Caculated_Bounds=NamedRerouteUsage()
   Pre_Skinned_Normals=NamedRerouteUsage()
   Is Vehicle=@n86.R
   Is VehicleSide=@n86.R
   Is Mech=@n86.R
   Vehicle_Rotation
   UV_Rotation Bypass=NamedRerouteUsage()
   Mech_Rotation
   Uvs_Offset
   Offset-after rotation
   Uvs_TileRate
  ) {materialFunction=&/MF_Projection_Side_Normal}
  NamedRerouteDeclaration(None=@n87.UVs No Offsets/Tiling) {name=UVs_Raw, nodeColor=<rgba 0.0078 0 1 1>}
  n88 = NamedRerouteUsage()
  n89 = ScalarParameter() {defaultValue=5, group=Wrap_NeonBlob_Noise, parameterName=NeonBL_NoiseWarp_Scale}
  n90 = Panner(
   Coordinate=Multiply(A=NamedRerouteUsage(), B=Multiply(A=@n88, B=@n89))
   Speed=ComponentMask(None=VectorParameter() {group=Wrap_NeonBlob_Noise, parameterName=NeonBL_NoiseWarp_Speed, defaultValue=<rgba -0.05 -0.1 0 1>}.RGB) {g=true, r=true}
  )
  n91 = NamedRerouteUsage()
  NamedRerouteDeclaration(None=Multiply(
    A=TextureSample(UVs=@n90, Tex=@n91) {samplerType=SAMPLERTYPE_Masks}.A
    B=ScalarParameter() {defaultValue=0.025, group=Wrap_NeonBlob_Noise, parameterName=NeonBL_NoiseWarp_Amount}
   )) {name=Warp_NoiseA, nodeColor=<rgba 0 1 0.1328 1>}
  NamedRerouteDeclaration(None=TextureObjectParameter() {group=Wrap_NeonBlob, parameterName=NeonBL__Tex_Msk1, texture=&T_BlacksRGB}) {name=Tex_Msk1, nodeColor=<rgba 1 0.2344 0 1>}
  NamedRerouteDeclaration(None=TextureObjectParameter() {group=Wrap_NeonBlob, parameterName=NeonBL__Tex_Msk2, texture=&T_BlacksRGB}) {name=Tex_Msk2, nodeColor=<rgba 1 0.2344 0 1>}
  n92 = NamedRerouteUsage()
  n93 = Multiply(A=NamedRerouteUsage(), B=ScalarParameter() {defaultValue=1.5, group=Wrap_NeonBlob_Noise, parameterName=NeonBL_Noise_Scale})
  n94 = VectorParameter() {group=Wrap_NeonBlob_Noise, parameterName=NeonBL_Noise_Speed, defaultValue=<rgba -0.03 0.004 0 1>}
  n95 = TextureSample(UVs=Panner(Coordinate=Multiply(A=@n92, B=@n93), Speed=ComponentMask(None=@n94.RGB) {g=true, r=true}), Tex=NamedRerouteUsage()) {samplerType=SAMPLERTYPE_Masks}
  NamedRerouteDeclaration(None=@n95.A) {name=Nose_Color_Warp, nodeColor=<rgba 1 0.9141 0 1>}
  n96 = MaterialFunctionCall(Color=Abs(None=NamedRerouteUsage())) {materialFunction=&SplitComponents}
  n97 = MaterialFunctionCall(Float4=NamedRerouteUsage()) {materialFunction=&BreakOutFloat4Components}
  n98 = LinearInterpolate(A=@n96.B, B=@n96.B, Alpha=@n97.G)
  n99 = MaterialFunctionCall(In=LinearInterpolate(A=LinearInterpolate(A=@n96.R, B=@n98, Alpha=@n97.R), B=@n96.G, Alpha=@n97.B), Contrast=@n78) {materialFunction=&CheapContrast}
  NamedRerouteDeclaration(None=@n99.Result) {name=Tri_Mask, nodeColor=<rgba 1 0.4688 0 1>}
  n100 = Constant() {r=0.5}
  n101 = Constant() {r=1}
  n102 = Constant() {r=0.5}
  n103 = Constant() {r=0.25}
  n104 = SetMaterialAttributes(
   MaterialAttributes=NamedRerouteUsage()
   BaseColor=LinearInterpolate(A=@n10.RGB, B=Multiply(A=@n10.RGB, B=@n62), Alpha=@n63)
   Roughness=LinearInterpolate(
    A=ScalarParameter() {defaultValue=0.5, group=Wrap_NeonBlob, parameterName=NeonBL_Roughness_Bck}
    B=ScalarParameter() {defaultValue=0.25, group=Wrap_NeonBlob, parameterName=NeonBL_Roughness_Blobs}
    Alpha=@n50
   )
   Specular=LinearInterpolate(
    A=ScalarParameter() {defaultValue=0.25, group=Wrap_NeonBlob, parameterName=NeonBL_Spec_Bck}
    B=ScalarParameter() {defaultValue=0.5, group=Wrap_NeonBlob, parameterName=NeonBL_Spec_Blobs}
    Alpha=@n50
   )
   Metallic=ScalarParameter() {group=Wrap_NeonBlob, parameterName=NeonBL_Metalic}
   Normal=Add(A=NamedRerouteUsage(), B=NamedRerouteUsage())
   EmissiveColor=Multiply(
    A=@n64
    B=MaterialFunctionCall(Value_Morning=@n100, Value_Day=@n101, Value_Evening=@n102, Value_Night=@n103) {materialFunction=&Scalar_Time_of_Day}.Result
   )
  ) {attributeSetTypes [
    #2
    #3
    #5
    #4
    #1
    #6
   ]}
  NamedRerouteDeclaration(None=@n104) {name=Mat_Base, nodeColor=<rgba 0 0.3906 1 1>}
  n105 = Add(A=@n95.A) {constB=-0.5}
  NamedRerouteDeclaration(None=Multiply(A=Power(Base=@n105), B=ScalarParameter() {defaultValue=3, group=Wrap_NeonBlob_Noise, parameterName=NeonBL_Noise_Color_Strength})) {name=Nose_Color_Warp_B, nodeColor=<rgba 1 0.9141 0 1>}
  n106 = VectorParameter() {group=Wrap_NeonBlob_Uvs, parameterName=NeonBL_Box_Offsest_Vehicle, defaultValue=<rgba -0.5 0.5 0 1>}
  n107 = VectorParameter() {group=Wrap_NeonBlob_Uvs, parameterName=NeonBL_Box_Offsest_VehicleSide, defaultValue=<rgba -0.5 0.5 0 1>}
  n108 = MaterialFunctionCall(Float4=NamedRerouteUsage()) {materialFunction=&BreakOutFloat4Components}
  n109 = LinearInterpolate(
   A=LinearInterpolate(
    A=VectorParameter() {group=Wrap_NeonBlob_Uvs, parameterName=NeonBL_Box_Offsest_Weapon, defaultValue=<rgba -1.2 0.5 0 1>}.RGB
    B=LinearInterpolate(A=@n106.RGB, B=@n107.RGB, Alpha=@n108.G)
    Alpha=@n108.R
   )
   B=LinearInterpolate(
    A=VectorParameter() {group=Wrap_NeonBlob_Uvs, parameterName=NeonBL_Box_Offsest_Mech, defaultValue=<rgba -0.5 0.5 0 1>}.RGB
    B=VectorParameter() {group=Wrap_NeonBlob_Uvs, parameterName=NeonBL_Box_Offsest_Character, defaultValue=<rgba -0.5 0.5 0 1>}.RGB
    Alpha=@n108.A
   )
   Alpha=@n108.B
  )
  n110 = VectorParameter() {group=Wrap_NeonBlob_Uvs, parameterName=NeonBL_Box_Offsest_Backpack, defaultValue=<rgba -0.5 0.5 0 1>}
  n111 = MaterialFunctionCall(Float2=NamedRerouteUsage()) {materialFunction=&BreakOutFloat2Components}
  NamedRerouteDeclaration(None=LinearInterpolate(
    A=LinearInterpolate(A=@n109, B=@n110.RGB, Alpha=@n111.R)
    B=VectorParameter() {group=Wrap_NeonBlob_Uvs, parameterName=NeonBL_Box_Offsest_Pickaxe, defaultValue=<rgba -0.5 0.5 0 1>}.RGB
    Alpha=@n111.G
   )) {name=UVs_Box_Offset, nodeColor=<rgba 0 0.625 1 1>}
  NamedRerouteDeclaration(None=@n87.UVs) {name=Uvs_Rotation, nodeColor=<rgba 1 0 0.3984 1>}
  n112 = ScalarParameter() {defaultValue=0.05, group=Wrap_NeonBlob_CC, parameterName=NeonGL_Curavture_Color_Strength_Vehicle, sortPriority=150}
  n113 = ScalarParameter() {defaultValue=0.05, group=Wrap_NeonBlob_CC, parameterName=NeonGL_Curavture_Color_Strength_Vehicle_Side, sortPriority=150}
  n114 = MaterialFunctionCall(Float4=NamedRerouteUsage()) {materialFunction=&BreakOutFloat4Components}
  n115 = LinearInterpolate(
   A=LinearInterpolate(
    A=ScalarParameter() {defaultValue=0.05, group=Wrap_NeonBlob_CC, parameterName=NeonGL_Curavture_Color_Strength_Weapon, sortPriority=150}
    B=LinearInterpolate(A=@n112, B=@n113, Alpha=@n114.G)
    Alpha=@n114.R
   )
   B=LinearInterpolate(
    A=ScalarParameter() {defaultValue=0.05, group=Wrap_NeonBlob_CC, parameterName=NeonGL_Curavture_Color_Strength_Mech, sortPriority=150}
    B=ScalarParameter() {defaultValue=0.05, group=Wrap_NeonBlob_CC, parameterName=NeonGL_Curavture_Color_Strength_Character, sortPriority=150}
    Alpha=@n114.A
   )
   Alpha=@n114.B
  )
  n116 = ScalarParameter() {defaultValue=0.05, group=Wrap_NeonBlob_CC, parameterName=NeonGL_Curavture_Color_Strength_Backpack, sortPriority=150}
  n117 = MaterialFunctionCall(Float2=NamedRerouteUsage()) {materialFunction=&BreakOutFloat2Components}
  NamedRerouteDeclaration(None=LinearInterpolate(
    A=LinearInterpolate(A=@n115, B=@n116, Alpha=@n117.R)
    B=ScalarParameter() {defaultValue=0.05, group=Wrap_NeonBlob_CC, parameterName=NeonGL_Curavture_Color_Strength_Pickaxe, sortPriority=150}
    Alpha=@n117.G
   )) {name=CC_Edge_Strength, nodeColor=<rgba 0 0.7188 1 1>}
  NamedRerouteDeclaration(None=@n1.A) {name=Msk_Gmap, nodeColor=<rgba 0 1 0.7656 1>}
  n118 = ScalarParameter() {group=Wrap_NeonBlob_Uvs, parameterName=NeonBL_Rotation_Vehicle, sortPriority=150}
  n119 = ScalarParameter() {group=Wrap_NeonBlob_Uvs, parameterName=NeonBL_Rotation_Vehicle_Side, sortPriority=150}
  n120 = MaterialFunctionCall(Float4=NamedRerouteUsage()) {materialFunction=&BreakOutFloat4Components}
  n121 = LinearInterpolate(
   A=LinearInterpolate(
    A=ScalarParameter() {group=Wrap_NeonBlob_Uvs, parameterName=NeonBL_Rotation_Weapon, sortPriority=150}
    B=LinearInterpolate(A=@n118, B=@n119, Alpha=@n120.G)
    Alpha=@n120.R
   )
   B=LinearInterpolate(
    A=ScalarParameter() {group=Wrap_NeonBlob_Uvs, parameterName=NeonBL_Rotation_Mech, sortPriority=150}
    B=ScalarParameter() {group=Wrap_NeonBlob_Uvs, parameterName=NeonBL_Rotation_Character, sortPriority=150}
    Alpha=@n120.A
   )
   Alpha=@n120.B
  )
  n122 = ScalarParameter() {group=Wrap_NeonBlob_Uvs, parameterName=NeonBL_Rotation_Backpack, sortPriority=150}
  n123 = MaterialFunctionCall(Float2=NamedRerouteUsage()) {materialFunction=&BreakOutFloat2Components}
  NamedRerouteDeclaration(None=LinearInterpolate(
    A=LinearInterpolate(A=@n121, B=@n122, Alpha=@n123.R)
    B=ScalarParameter() {group=Wrap_NeonBlob_Uvs, parameterName=NeonBL_Rotation_Pickaxe, sortPriority=150}
    Alpha=@n123.G
   )) {name=UVs_Rotation_Offsets, nodeColor=<rgba 0 0.7188 1 1>}
  n124 = ScalarParameter() {defaultValue=1, group=Wrap_NeonBlob_Uvs, parameterName=NeonBL_Scale_Mast_Vehicle, sortPriority=150}
  n125 = ScalarParameter() {defaultValue=1, group=Wrap_NeonBlob_Uvs, parameterName=NeonBL_Scale_Mast_Vehicle_Side, sortPriority=150}
  n126 = MaterialFunctionCall(Float4=NamedRerouteUsage()) {materialFunction=&BreakOutFloat4Components}
  n127 = LinearInterpolate(
   A=LinearInterpolate(
    A=ScalarParameter() {defaultValue=1, group=Wrap_NeonBlob_Uvs, parameterName=NeonBL_Scale_Mast_Weapon, sortPriority=150}
    B=LinearInterpolate(A=@n124, B=@n125, Alpha=@n126.G)
    Alpha=@n126.R
   )
   B=LinearInterpolate(
    A=ScalarParameter() {defaultValue=1, group=Wrap_NeonBlob_Uvs, parameterName=NeonBL_Scale_Mast_Rotation_Mech, sortPriority=150}
    B=ScalarParameter() {defaultValue=1, group=Wrap_NeonBlob_Uvs, parameterName=NeonBL_Scale_Mast_Character, sortPriority=150}
    Alpha=@n126.A
   )
   Alpha=@n126.B
  )
  n128 = ScalarParameter() {defaultValue=1, group=Wrap_NeonBlob_Uvs, parameterName=NeonBL_Scale_Mast_Backpack, sortPriority=150}
  n129 = MaterialFunctionCall(Float2=NamedRerouteUsage()) {materialFunction=&BreakOutFloat2Components}
  NamedRerouteDeclaration(None=LinearInterpolate(
    A=LinearInterpolate(A=@n127, B=@n128, Alpha=@n129.R)
    B=ScalarParameter() {defaultValue=1, group=Wrap_NeonBlob_Uvs, parameterName=NeonBL_Scale_Mast_Pickaxe, sortPriority=150}
    Alpha=@n129.G
   )) {name=UVs_Scale_Master, nodeColor=<rgba 0 0.7188 1 1>}
  NamedRerouteDeclaration(None=@n99.Result) {name=Tri_Mask_1, nodeColor=<rgba 1 0.4688 0 1>}
  NamedRerouteDeclaration(None=FunctionInput(Preview=Constant4Vector().RGBA) {bUsePreviewValueAsDefault=true, inputName=IsState, inputType=FunctionInput_Vector4, sortPriority=4}) {name=IsState, nodeColor=<rgba 0.1497 1 0.1633 1>}
  NamedRerouteDeclaration(None=FunctionInput(Preview=Constant2Vector()) {bUsePreviewValueAsDefault=true, inputName=IsState_B, inputType=FunctionInput_Vector2, sortPriority=4}) {name=IsState_B, nodeColor=<rgba 0.1497 1 0.1633 1>}
  n130 = PreSkinnedLocalBounds()
  n131 = Subtract(A=VertexInterpolator(VS=PreSkinnedPosition()).PS, B=@n130.Min)
  n132 = MaterialFunctionCall(Color=Subtract(A=@n130.Max, B=@n130.Max)) {materialFunction=&SplitComponents}
  n133 = Max(A=Max(A=@n132.R, B=@n132.R), B=@n132.B)
  NamedRerouteDeclaration(None=FunctionInput(Preview=Divide(A=@n131, B=@n133)) {bUsePreviewValueAsDefault=true, inputName=Calcated_Bounds, sortPriority=11}) {name=Caclcated_Bounds, nodeColor=<rgba 0 0.9297 1 1>}
  n134 = VectorParameter() {group=Wrap_NeonBlob_Uvs, parameterName=NeonBL_Box_Scale_Vehicle, defaultValue=<rgba 0.1 0.1 0 1>}
  n135 = VectorParameter() {group=Wrap_NeonBlob_Uvs, parameterName=NeonBL_Box_Scale_VehicleSide, defaultValue=<rgba 0.1 0.1 0 1>}
  n136 = MaterialFunctionCall(Float4=NamedRerouteUsage()) {materialFunction=&BreakOutFloat4Components}
  n137 = LinearInterpolate(
   A=LinearInterpolate(
    A=VectorParameter() {group=Wrap_NeonBlob_Uvs, parameterName=NeonBL_Box_Scale_Weapon, defaultValue=<rgba 0.5 1 0 1>}.RGB
    B=LinearInterpolate(A=@n134.RGB, B=@n135.RGB, Alpha=@n136.G)
    Alpha=@n136.R
   )
   B=LinearInterpolate(
    A=VectorParameter() {group=Wrap_NeonBlob_Uvs, parameterName=NeonBL_Box_Scale_Mech, defaultValue=<rgba 0.1 0.1 0 1>}.RGB
    B=VectorParameter() {group=Wrap_NeonBlob_Uvs, parameterName=NeonBL_Box_Scale_Character, defaultValue=<rgba 0.1 0.1 0 1>}.RGB
    Alpha=@n136.A
   )
   Alpha=@n136.B
  )
  n138 = VectorParameter() {group=Wrap_NeonBlob_Uvs, parameterName=NeonBL_Box_Scale_Backpack, defaultValue=<rgba 0.1 0.1 0 1>}
  n139 = MaterialFunctionCall(Float2=NamedRerouteUsage()) {materialFunction=&BreakOutFloat2Components}
  NamedRerouteDeclaration(None=LinearInterpolate(
    A=LinearInterpolate(A=@n137, B=@n138.RGB, Alpha=@n139.R)
    B=VectorParameter() {group=Wrap_NeonBlob_Uvs, parameterName=NeonBL_Box_Scale_Pickaxe, defaultValue=<rgba 0.1 0.1 0 1>}.RGB
    Alpha=@n139.G
   )) {name=UVs_Box_Scale, nodeColor=<rgba 0 0.625 1 1>}
  n140 = PreSkinnedNormal()
  NamedRerouteDeclaration(None=FunctionInput(Preview=VertexInterpolator(VS=@n140).PS) {bUsePreviewValueAsDefault=true, inputName=Pre-Skinned Local Normal, sortPriority=10}) {name=PreSkin_Normal, nodeColor=<rgba 0.8281 1 0 1>}
  NamedRerouteDeclaration(None=@n9.Normal) {name=Root_Normal, nodeColor=<rgba 0.4062 0 1 1>}
  NamedRerouteUsage()
  NamedRerouteUsage()
  NamedRerouteUsage()
  NamedRerouteUsage()
  Power(Base=@n85, Exp=ScalarParameter() {defaultValue=1, group=Wrap_NeonBlob, parameterName=NeonGL_Normal_Falloff})
  ScalarParameter() {defaultValue=0.05, group=Wrap_NeonBlob_CC, parameterName=NeonGL_Curavture_Color_Strength}
  VertexNormalWS()
 ]}
