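# Material function for the Neon Graffiti Lava wrap, layered over an input material through primary and secondary masks behind a static switch.
# A noise-warped box falloff, panning drips and two projected logos form a height mask, remapped by curve-atlas rows into lava and glow.
# Emissive combines lava, embers and gradient background, scaled by time of day; normals derive from the height.
# IsState inputs select UV tuning for weapons, vehicles, mechs, characters, backpacks and pickaxes.
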MaterialFunction /Game/Weapons/WeaponSkins/Wraps/MaterialFunctions/MF_Wrap_NeonGraffitiLava
graph {nodes [
  n1 = VectorParameter() {group=Wrap_NeonGraffitiLava_Embers, parameterName=NeonGL_Embers_Color_A, defaultValue=<rgba 0.0625 0.0163 0.0037 1>}
  n2 = NamedRerouteUsage()
  n3 = NamedRerouteUsage()
  n4 = Multiply(A=@n1.RGB, B=TextureSample(UVs=@n2, Tex=@n3).R)
  n5 = VectorParameter() {group=Wrap_NeonGraffitiLava_Embers, parameterName=NeonGL_Embers_Speed_B, defaultValue=<rgba -0.013 -0.08 0 1>}
  n6 = Panner(
   Coordinate=Multiply(A=@n2, B=ScalarParameter() {defaultValue=3, group=Wrap_NeonGraffitiLava_Embers, parameterName=NeonGL_Embers_TileRate_B})
   Speed=ComponentMask(None=@n5.RGB) {g=true, r=true}
  )
  n7 = VectorParameter() {group=Wrap_NeonGraffitiLava_Embers, parameterName=NeonGL_Embers_Color_B, defaultValue=<rgba 0.7344 0.261 0.0376 1>}
  Add(A=@n4, B=Multiply(A=TextureSample(UVs=@n6, Tex=@n3).R, B=@n7.RGB))
  n8 = MaterialFunctionCall(Float4=NamedRerouteUsage()) {materialFunction=&BreakOutFloat4Components}
  n9 = NamedRerouteUsage()
  n10 = Constant()
  n11 = NamedRerouteUsage()
  n12 = LinearInterpolate(
   A=Constant3Vector() {constant=<rgba 1 1 1 0>}
   B=Constant3Vector() {constant=<rgba -1 1 1 0>}
   Alpha=Max(A=Max(A=@n10, B=@n11), B=NamedRerouteUsage())
  )
  n13 = MaterialFunctionCall!(
   For_Safe_Zones=StaticBool() {value=true}
   Normal_Map
   Caculated_Bounds=NamedRerouteUsage()
   Pre_Skinned_Normals=NamedRerouteUsage()
   Is Vehicle=@n8.R
   Is VehicleSide=@n8.R
   Is Mech=@n8.R
   Vehicle_Rotation
   UV_Rotation Bypass=NamedRerouteUsage()
   Mech_Rotation
   Uvs_Offset=ComponentMask(None=@n9) {b=true, g=true, r=true}
   Offset-after rotation
   Uvs_TileRate=Multiply(A=ComponentMask(None=@n9) {a=true}, B=@n12)
  ) {materialFunction=&/MF_Projection_Side_Normal}
  n14 = Constant2Vector() {g=2, r=2}
  Add(A=Multiply(A=@n13.UVs, B=Divide(B=@n14)), B=Constant2Vector() {r=0.5})
  n15 = NamedRerouteUsage()
  n16 = NamedRerouteUsage()
  n17 = ScalarParameter() {defaultValue=0.02, group=Wrap_NeonGraffitiLava, parameterName=NeonGL_Lava_Noise_Amount_Logo}
  n18 = TextureSample(UVs=Add(A=Multiply(A=@n16, B=@n17) {constB=0.025}, B=@n13.UVs), Tex=@n15)
  n19 = Multiply(A=NamedRerouteUsage(), B=Multiply(A=TextureSample(UVs=@n13.UVs, Tex=@n15).R, B=@n18.A))
  n20 = Multiply(A=NamedRerouteUsage(), B=@n17) {constB=0.025}
  n21 = MaterialFunctionCall(Float4=NamedRerouteUsage()) {materialFunction=&BreakOutFloat4Components}
  n22 = NamedRerouteUsage()
  n23 = MaterialFunctionCall(
   For_Safe_Zones=StaticBool() {value=true}
   Caculated_Bounds=NamedRerouteUsage()
   Pre_Skinned_Normals=NamedRerouteUsage()
   Is Vehicle=@n21.R
   Is VehicleSide=@n21.R
   Is Mech=@n21.R
   UV_Rotation Bypass=NamedRerouteUsage()
   Uvs_Offset=ComponentMask(None=@n22) {b=true, g=true, r=true}
   Uvs_TileRate=Multiply(A=ComponentMask(None=@n22) {a=true}, B=@n12)
  ) {materialFunction=&MF_Projection_Side_Normal}
  n24 = Constant2Vector() {g=2, r=2}
  n25 = Add(A=Multiply(A=@n23.UVs, B=Divide(B=@n24)), B=Constant2Vector())
  n26 = NamedRerouteUsage()
  n27 = Multiply(A=TextureSample(UVs=Add(A=@n20, B=@n25), Tex=@n26).A, B=TextureSample(UVs=@n23.UVs, Tex=@n26).R)
  Add(A=@n19, B=Multiply(A=@n27) {constB=0.5})
  Add(A=@n19, B=@n27)
  n28 = Add(A=NamedRerouteUsage(), B=ComponentMask(None=NamedRerouteUsage()) {g=true, r=true}) {constB=-0.5}
  n29 = Subtract(A=Abs(None=Multiply(A=@n28) {constB=2}), B=ComponentMask(None=NamedRerouteUsage()) {g=true, r=true})
  n30 = MaterialFunctionCall(Vector 2=Max(A=@n29) {constB=0}) {materialFunction=/Engine/Functions/Engine_MaterialFunctions02/Utility/VectorLength}
  n31 = MaterialFunctionCall(Float2=@n29) {materialFunction=&BreakOutFloat2Components}
  n32 = Min(A=Max(A=@n31.R, B=@n31.R) {constB=0}) {constB=0}
  n33 = OneMinus(None=Divide(A=Add(A=@n30.V2 Length, B=@n32), B=ScalarParameter() {defaultValue=0.125, group=Wrap_NeonGraffitiLava, parameterName=NeonGL_Lava_Falloff}) {constB=0.1})
  n34 = Multiply(A=NamedRerouteUsage(), B=NamedRerouteUsage())
  n35 = VectorParameter() {group=Wrap_NeonGraffitiLava_UVs, parameterName=NeonGL_Lava_Noise_Speed, defaultValue=<rgba 0.0025 -0.01 0 1>}
  n36 = ComponentMask(None=@n35.RGB) {g=true, r=true}
  n37 = NamedRerouteUsage()
  n38 = TextureSample(UVs=Panner(Coordinate=@n34, Speed=@n36), Tex=@n37) {samplerType=SAMPLERTYPE_LinearGrayscale}
  n39 = Add(
   A=Saturate(None=Power(Base=@n33))
   B=Multiply(
    A=NamedRerouteDeclaration(None=@n38.R) {name=Warp_Noise, nodeColor=<rgba 0.0078 0 1 1>}
    B=ScalarParameter() {defaultValue=0.4, group=Wrap_NeonGraffitiLava, parameterName=NeonGL_Lava_Noise_Amount}
   ) {constB=0.5}
  )
  n40 = NamedRerouteUsage()
  n41 = Multiply(
   A=TextureSample(UVs=@n34, Tex=@n37) {samplerType=SAMPLERTYPE_LinearGrayscale}.R
   B=ScalarParameter() {defaultValue=0.025, group=Wrap_NeonGraffitiLava, parameterName=NeonGL_Lava_Noise_Amount_Drip}
  )
  n42 = NamedRerouteUsage()
  n43 = Constant() {r=0.5}
  n44 = NamedRerouteUsage()
  n45 = NamedRerouteUsage()
  n46 = VectorParameter() {group=Wrap_NeonGraffitiLava_UVs, parameterName=NeonGL_Lava_Drip_Speed, defaultValue=<rgba 0.0025 -0.015 0 1>}
  n47 = Panner(
   Coordinate=Multiply(
    A=MaterialFunctionCall(UVs=@n42, Rotation Center=@n43, "Rotation Angle (0-1)"=@n44) {materialFunction=&CustomRotator}.Rotated Values
    B=ComponentMask(None=@n45) {g=true, r=true}
   )
   Speed=ComponentMask(None=@n46.RGB) {g=true, r=true}
  )
  n48 = Multiply(A=TextureSample(UVs=Add(A=@n41, B=@n47), Tex=NamedRerouteUsage()).A, B=NamedRerouteUsage())
  n49 = Multiply(A=OneMinus(None=@n39), B=Multiply(A=Multiply(A=@n48) {constB=0.75}, B=OneMinus(None=@n19)))
  n50 = Add(A=@n39, B=Multiply(A=OneMinus(None=@n40), B=Add(A=@n49, B=@n19)))
  n51 = CurveAtlasRowParameter(CurveTime=@n50) {
   atlas=/Game/Weapons/WeaponSkins/Wraps/Textures/CruveAtlas_Wraps_01
   curve=/Game/Weapons/WeaponSkins/Wraps/Textures/0_ColorCurves/CC_NGL_Shape
   defaultValue=75
   group=Wrap_NeonGraffitiLava
   parameterName="NeonGL_BG_ColorCurve_Height\\Msk"
  }
  MaterialExpressionMaterialFunctionCall_8 = MaterialFunctionCall!(
   Index=ScalarParameter() {defaultValue=75, group=Wrap_NeonGraffitiLava, parameterName="NeonGL_BG_ColorCurve_Height\\Msk  Index"}
   CCinput=@n51
   CCinput_Alpha=@n51
   UV=@n50
  ) {materialFunction=&/MF_Global_CC_Tex}
  n52 = Convert(None=@MaterialExpressionMaterialFunctionCall_8.3V) {
   nodeName=Break Float3
   convertInputs [{
     type=Vector3
     expressionInput {
      expression=&self:MaterialExpressionMaterialFunctionCall_8
      inputName=None
      mask=0
      maskA=0
      maskB=0
      maskG=0
      maskR=0
      outputIndex=0
     }
    }]
   convertOutputs [{type=Scalar}, {type=Scalar}, {type=Scalar}]
  }
  n53 = MaterialFunctionCall!(Value=@n52) {materialFunction=/Engine/Functions/Engine_MaterialFunctions03/Procedurals/PreparePerturbNormalHQ}
  n54 = Subtract(A=@n53.Value, B=@n53.Value one pixel right)
  n55 = Subtract(A=@n53.Value, B=@n53.Value one pixel down)
  AppendVector(
   A=Multiply(A=AppendVector(A=@n54, B=@n55), B=ScalarParameter() {defaultValue=5, group=Wrap_NeonGraffitiLava, parameterName=NeonGL_Normal_Strength_B}) {constB=5}
   B=Constant()
  )
  Constant4Vector()
  n56 = TextureSample() {samplerType=SAMPLERTYPE_Masks, texture=/Game/Weapons/FORT_Rifles/Textures/T_Scar_CM}
  n57 = FunctionInput(Preview=@n56.R) {bUsePreviewValueAsDefault=true, inputName=Mask Primary, inputType=FunctionInput_Scalar, sortPriority=24, previewValue=<1 0 0 1>}
  n58 = FunctionInput(Preview=@n56.B) {bUsePreviewValueAsDefault=true, inputName=Mask Secondary, inputType=FunctionInput_Scalar, sortPriority=24, previewValue=<1 0 0 1>}
  n59 = Add(A=@n57, B=@n58)
  n60 = NamedRerouteUsage()
  n61 = BlendMaterialAttributes(
   A=BlendMaterialAttributes(A=NamedRerouteUsage(), B=NamedRerouteUsage(), Alpha=Saturate(None=@n59)) {vertexAttributeBlendType=UseA}
   B=NamedRerouteUsage()
   Alpha=Multiply(A=NamedRerouteDeclaration(None=@n58) {name=Msk_Accent, nodeColor=<rgba 0 1 0.7656 1>}, B=OneMinus(None=@n60))
  ) {vertexAttributeBlendType=UseA}
  n62 = TextureSample() {samplerType=SAMPLERTYPE_Normal, texture=/Game/Weapons/FORT_Rifles/T_SCAR_N}
  n63 = Constant3Vector() {constant=<rgba 1 0 0 0>}
  n64 = GetMaterialAttributes(None=FunctionInput(Preview=SetMaterialAttributes!(MaterialAttributes, Normal=@n62.RGB, BaseColor=@n63) {attributeSetTypes [#1, #2]}) {inputName=Material, inputType=FunctionInput_MaterialAttributes}) {attributeGetTypes [
    #2
    #3
    #1
    #4
   ]}
  FunctionOutput(None=StaticSwitchParameter(True=@n61, False=NamedRerouteDeclaration(None=@n64.MaterialAttributes) {name=Root_Material, nodeColor=<rgba 0.7812 1 0 1>}) {group=Customization_Wrap, parameterName=Wrap_NeonGraffitiLava})
  Multiply(A=@n18.A)
  Multiply(A=@n53.FilterWidth)
  n65 = VectorParameter() {group=Wrap_NeonGraffitiLava_UVs, parameterName=NeonGL_Lava_Drip_TileRate, defaultValue=<rgba 3 2 0 1>}
  Multiply(A=@n65.RGB, B=@n65.RGB)
  n66 = VectorParameter() {group=Wrap_NeonGraffitiLava_Accent, parameterName=NeonGL_Accent_BaseColor, defaultValue=<rgba 0 0.003 0.0004 1>}
  n67 = SetMaterialAttributes!(
   MaterialAttributes=NamedRerouteUsage()
   BaseColor=@n66.RGB
   Roughness=ScalarParameter() {defaultValue=0.3, group=Wrap_NeonGraffitiLava_Accent, parameterName=NeonGL_Accent_Roughness}
   Specular=ScalarParameter() {defaultValue=0.5, group=Wrap_NeonGraffitiLava_Accent, parameterName=NeonGL_Accent_Spec}
   Metallic=ScalarParameter() {defaultValue=0.5, group=Wrap_NeonGraffitiLava_Accent, parameterName=NeonGL_Accent_Metalic}
   Normal
   EmissiveColor
  ) {attributeSetTypes [
    #2
    #3
    #5
    #4
    #1
    #6
   ]}
  NamedRerouteDeclaration(None=@n67) {name=Mat_Accent, nodeColor=<rgba 0 0.3906 1 1>}
  n68 = FunctionInput(Preview=@n56.G) {bUsePreviewValueAsDefault=true, inputName=Mask Crevise, inputType=FunctionInput_Scalar, sortPriority=24, previewValue=<1 0 0 1>}
  NamedRerouteDeclaration(None=@n68) {name=Msk_Curvature, nodeColor=<rgba 0 1 0.7422 1>}
  n69 = TextureObjectParameter() {group=Wrap_NeonGraffitiLava, parameterName=NeonGL__Tex_Msk, sortPriority=10, texture=&T_BlacksRGB}
  NamedRerouteDeclaration(None=@n69) {name=Tex_Msk, nodeColor=<rgba 1 0.5859 0 1>}
  n70 = MaterialFunctionCall(Color=Abs(None=NamedRerouteUsage())) {materialFunction=&SplitComponents}
  n71 = MaterialFunctionCall(Float4=NamedRerouteUsage()) {materialFunction=&BreakOutFloat4Components}
  n72 = LinearInterpolate(A=@n70.R, B=LinearInterpolate(A=@n70.B, B=@n70.B, Alpha=@n71.G), Alpha=@n71.R)
  NamedRerouteDeclaration(None=MaterialFunctionCall(
    In=LinearInterpolate(A=@n72, B=@n70.G, Alpha=@n71.B)
    Contrast=ScalarParameter() {group=Wrap_NeonGraffitiLava, parameterName=NeonGL_Ti_Blend_Contrast, sortPriority=0}
   ) {materialFunction=&CheapContrast}.Result) {name=Tri_Mask, nodeColor=<rgba 1 0.4688 0 1>}
  n73 = TextureObjectParameter() {group=Wrap_NeonGraffitiLava, parameterName=NeonGL__Tex_Logo, sortPriority=10, texture=&T_BlacksRGB}
  NamedRerouteDeclaration(None=@n73) {name=Tex_Logos, nodeColor=<rgba 1 0.5859 0 1>}
  n74 = Multiply(
   A=Transform(None=MaterialFunctionCall!(
     Bump one pixel right=@n53.Value one pixel right
     Bump center=@n53.Value one pixel right
     Bump one pixel down=@n53.Value one pixel right
     WS Normal
    ) {materialFunction=/Engine/Functions/Engine_MaterialFunctions03/Procedurals/PerturbNormalHQ}.WS Normal) {transformSourceType=TRANSFORMSOURCE_World, transformType=TRANSFORM_Tangent}
   B=Constant3Vector() {constant=<rgba 1 1 0 0>}
  )
  n75 = ScalarParameter() {defaultValue=1, group=Wrap_NeonGraffitiLava, parameterName=NeonGL_Normal_Falloff}
  NamedRerouteDeclaration(None=Multiply(A=Power(Base=@n74, Exp=@n75), B=ScalarParameter() {defaultValue=5, group=Wrap_NeonGraffitiLava, parameterName=NeonGL_Normal_Strength})) {name=Generated_Normals, nodeColor=<rgba 0.6406 0 1 1>}
  n76 = ScalarParameter() {group=Wrap_NeonGraffitiLava_UVs, parameterName=NeonGL_Lava_Rotation_Vehicle}
  n77 = ScalarParameter() {group=Wrap_NeonGraffitiLava_UVs, parameterName=NeonGL_Lava_Rotation_Vehicle_Side}
  n78 = MaterialFunctionCall(Float4=NamedRerouteUsage()) {materialFunction=&BreakOutFloat4Components}
  n79 = LinearInterpolate(
   A=LinearInterpolate(
    A=ScalarParameter() {group=Wrap_NeonGraffitiLava_UVs, parameterName=NeonGL_Lava_Rotation_Weapon}
    B=LinearInterpolate(A=@n76, B=@n77, Alpha=@n78.G)
    Alpha=@n78.R
   )
   B=LinearInterpolate(
    A=ScalarParameter() {group=Wrap_NeonGraffitiLava_UVs, parameterName=NeonGL_Lava_Rotation_Mech}
    B=ScalarParameter() {group=Wrap_NeonGraffitiLava_UVs, parameterName=NeonGL_Lava_Rotation_Character}
    Alpha=@n78.A
   )
   Alpha=@n78.B
  )
  n80 = ScalarParameter() {group=Wrap_NeonGraffitiLava_UVs, parameterName=NeonGL_Lava_Rotation_Backpack}
  n81 = MaterialFunctionCall(Float2=NamedRerouteUsage()) {materialFunction=&BreakOutFloat2Components}
  NamedRerouteDeclaration(None=LinearInterpolate(
    A=LinearInterpolate(A=@n79, B=@n80, Alpha=@n81.R)
    B=ScalarParameter() {group=Wrap_NeonGraffitiLava_UVs, parameterName=NeonGL_Lava_Rotation_Pickaxe}
    Alpha=@n81.G
   )) {name=UVs_Lava_Rotation, nodeColor=<rgba 0 0.7188 1 1>}
  n82 = VectorParameter() {group=Wrap_NeonGraffitiLava, parameterName=NeonGL_BaseColor, sortPriority=5, defaultValue=<rgba 0 0 0 1>}
  n83 = ScalarParameter() {defaultValue=0.7, group=Wrap_NeonGraffitiLava, parameterName=NeonGL_Roughness_Lava_edge, sortPriority=5}
  n84 = ScalarParameter() {defaultValue=0.35, group=Wrap_NeonGraffitiLava, parameterName=NeonGL_Roughness_Lava, sortPriority=5}
  n85 = NamedRerouteUsage()
  n86 = NamedRerouteUsage()
  n87 = NamedRerouteUsage()
  n88 = Multiply(
   A=MaterialFunctionCall(
    A=VectorParameter() {group=Wrap_NeonGraffitiLava, parameterName=NeonGL_Lava_Color_A, defaultValue=<rgba 0.3333 0.0129 0 1>}.RGB
    B=VectorParameter() {group=Wrap_NeonGraffitiLava, parameterName=NeonGL_Lava_Color_B, defaultValue=<rgba 0.6719 0.1251 0 1>}.RGB
    C=VectorParameter() {group=Wrap_NeonGraffitiLava, parameterName=NeonGL_Lava_Color_C, defaultValue=<rgba 1 0.4809 0 1>}.RGB
    Alpha=LinearInterpolate(Alpha=@n87) {constA=-0.1, constB=1.5}
   ) {materialFunction=&3ColorBlend}.Result
   B=NamedRerouteUsage()
  )
  n89 = Power(Base=NamedRerouteUsage())
  n90 = Multiply(
   A=Add(A=Multiply(A=@n88, B=@n89), B=NamedRerouteUsage())
   B=ScalarParameter() {defaultValue=1, group=Wrap_NeonGraffitiLava, parameterName=NeonGL_Lava_Brightness}
  )
  n91 = Multiply(A=@n90, B=ScalarParameter() {defaultValue=5, group=Wrap_NeonGraffitiLava, parameterName=NeonGL_Lava_Edge_Brightness}) {constB=2}
  n92 = NamedRerouteUsage()
  n93 = Saturate(None=Multiply(A=Add(A=@n92) {constB=-0.5}) {constB=2})
  n94 = LinearInterpolate(A=@n90, B=Multiply(A=@n90) {constB=0.5}, Alpha=@n93) {constB=0}
  n95 = Saturate(None=Multiply(A=Add(A=@n92) {constB=0}) {constB=2})
  n96 = NamedRerouteUsage()
  n97 = VectorParameter() {group=Wrap_NeonGraffitiLava, parameterName=NeonGL_Lava_Color_Glow, defaultValue=<rgba 0.3333 0.0129 0 1>}
  n98 = Clamp(None=Add(A=LinearInterpolate(A=@n91, B=@n94, Alpha=@n95), B=Multiply(A=@n96, B=@n97.RGB)) {constB=0}) {maxDefault=10}
  n99 = VectorParameter() {group=Wrap_NeonGraffitiLava_Bck, parameterName=NeonGL_BG_Color_A, defaultValue=<rgba 0 0 0 1>}
  n100 = VectorParameter() {group=Wrap_NeonGraffitiLava_Bck, parameterName=NeonGL_BG_Color_B, defaultValue=<rgba 0.4427 0.0033 0 1>}
  n101 = VectorParameter() {group=Wrap_NeonGraffitiLava_Bck, parameterName=NeonGL_BG_Color_C, defaultValue=<rgba 1 0.1831 0 1>}
  n102 = Constant3Vector() {constant=<rgba 0 1 0 0>}
  n103 = MaterialFunctionCall(Float4=NamedRerouteUsage()) {materialFunction=&BreakOutFloat4Components}
  n104 = LinearInterpolate(A=LinearInterpolate(A=@n102, B=LinearInterpolate(B=@n102, Alpha=@n103.R), Alpha=@n103.G), B=@n102, Alpha=@n103.B)
  n105 = MaterialFunctionCall(Float4=NamedRerouteUsage()) {materialFunction=&BreakOutFloat4Components}
  n106 = MaterialFunctionCall(Caculated_Bounds=NamedRerouteUsage(), Pre_Skinned_Normals=NamedRerouteUsage(), Is Vehicle=@n105.R, Is VehicleSide=@n105.R, Is Mech=@n105.R) {materialFunction=&MF_Projection_Side_Normal}
  n107 = MaterialFunctionCall(
   UVs=Add(
    A=ComponentMask(None=@n104) {g=true, r=true}
    B=NamedRerouteDeclaration(None=@n106.UVs No Offsets/Tiling) {name=UVs_Raw, nodeColor=<rgba 0.0078 0 1 1>}
   )
   Rotation Center=Constant() {r=0.5}
   "Rotation Angle (0-1)"=NamedRerouteUsage()
  ) {materialFunction=&CustomRotator}
  n108 = Multiply(
   A=OneMinus(None=ComponentMask(None=@n107.Rotated Values) {r=true})
   B=ScalarParameter() {defaultValue=0.8, group=Wrap_NeonGraffitiLava_Bck, parameterName=NeonGL_BG_Brightness}
  )
  n109 = Multiply(
   A=Add(A=NamedRerouteUsage(), B=ScalarParameter() {defaultValue=-0.25, group=Wrap_NeonGraffitiLava_Bck, parameterName=NeonGL_BG_Edge_Offset}) {constB=-0.5}
   B=ScalarParameter() {defaultValue=-0.4, group=Wrap_NeonGraffitiLava_Bck, parameterName=NeonGL_BG_Edge_Brightness}
  ) {constB=2}
  n110 = Saturate(None=Power(Base=Add(A=@n108, B=@n109), Exp=ScalarParameter() {defaultValue=5, group=Wrap_NeonGraffitiLava_Bck, parameterName=NeonGL_BG_Falloff}))
  n111 = Multiply(
   A=Add(
    A=MaterialFunctionCall(A=@n99.RGB, B=@n100.RGB, C=@n101.RGB, Alpha=@n110) {materialFunction=&3ColorBlend}.Result
    B=NamedRerouteUsage()
   )
   B=OneMinus(None=NamedRerouteUsage())
  )
  n112 = Constant() {r=0.5}
  n113 = Constant() {r=1}
  n114 = Constant() {r=0.5}
  n115 = Constant() {r=0.25}
  n116 = SetMaterialAttributes(
   MaterialAttributes=NamedRerouteUsage()
   BaseColor=@n82.RGB
   Roughness=LinearInterpolate(
    A=ScalarParameter() {defaultValue=0.5, group=Wrap_NeonGraffitiLava, parameterName=NeonGL_Roughness, sortPriority=5}
    B=LinearInterpolate(A=@n83, B=@n84, Alpha=@n85) {constA=1, constB=0}
    Alpha=@n86
   )
   Specular=LinearInterpolate(
    A=ScalarParameter() {defaultValue=0.15, group=Wrap_NeonGraffitiLava, parameterName=NeonGL_Spec, sortPriority=5}
    B=ScalarParameter() {defaultValue=0.5, group=Wrap_NeonGraffitiLava, parameterName=NeonGL_Spec_Lava, sortPriority=5}
    Alpha=@n86
   )
   Metallic=Constant()
   Normal=Add(A=NamedRerouteUsage(), B=NamedRerouteUsage())
   EmissiveColor=Multiply(
    A=Add(A=@n98, B=@n111)
    B=MaterialFunctionCall(Value_Morning=@n112, Value_Day=@n113, Value_Evening=@n114, Value_Night=@n115) {materialFunction=&Scalar_Time_of_Day}.Result
   )
  ) {attributeSetTypes [
    #2
    #3
    #5
    #4
    #1
    #6
   ]}
  NamedRerouteDeclaration(None=@n116) {name=Material_Back, nodeColor=<rgba 0 0.3906 1 1>}
  n117 = VectorParameter() {group=Wrap_NeonGraffitiLava_UVs, parameterName=NeonGL_LogoA_Offsest_Weapon, sortPriority=100, defaultValue=<rgba 0.15 -0.01 0 0.18>}
  n118 = VectorParameter() {group=Wrap_NeonGraffitiLava_UVs, parameterName=NeonGL_LogoA_Offsest_Vehicle, sortPriority=100, defaultValue=<rgba 0 0 0 1>}
  n119 = AppendVector(A=@n118.RGB, B=@n118.RGB)
  n120 = VectorParameter() {group=Wrap_NeonGraffitiLava_UVs, parameterName=NeonGL_LogoA_Offsest_VehicleSide, sortPriority=100, defaultValue=<rgba 0 0 0 1>}
  n121 = AppendVector(A=@n120.RGB, B=@n120.RGB)
  n122 = MaterialFunctionCall(Float4=NamedRerouteUsage()) {materialFunction=&BreakOutFloat4Components}
  n123 = VectorParameter() {group=Wrap_NeonGraffitiLava_UVs, parameterName=NeonGL_LogoA_Offsest_Mech, sortPriority=100, defaultValue=<rgba 0 0 0 1>}
  n124 = VectorParameter() {group=Wrap_NeonGraffitiLava_UVs, parameterName=NeonGL_LogoA_Offsest_Character, sortPriority=100, defaultValue=<rgba 0 0 0 1>}
  n125 = LinearInterpolate(
   A=LinearInterpolate(A=AppendVector(A=@n117.RGB, B=@n117.RGB), B=LinearInterpolate(A=@n119, B=@n121, Alpha=@n122.G), Alpha=@n122.R)
   B=LinearInterpolate(A=AppendVector(A=@n123.RGB, B=@n123.RGB), B=AppendVector(A=@n124.RGB, B=@n124.RGB), Alpha=@n122.A)
   Alpha=@n122.B
  )
  n126 = VectorParameter() {group=Wrap_NeonGraffitiLava_UVs, parameterName=NeonGL_LogoA_Offsest_Backpack, sortPriority=100, defaultValue=<rgba 0 0 0 1>}
  n127 = AppendVector(A=@n126.RGB, B=@n126.RGB)
  n128 = MaterialFunctionCall(Float2=NamedRerouteUsage()) {materialFunction=&BreakOutFloat2Components}
  n129 = VectorParameter() {group=Wrap_NeonGraffitiLava_UVs, parameterName=NeonGL_LogoA_Offsest_Pickaxe, sortPriority=100, defaultValue=<rgba 0 0 0 1>}
  NamedRerouteDeclaration(None=LinearInterpolate(A=LinearInterpolate(A=@n125, B=@n127, Alpha=@n128.R), B=AppendVector(A=@n129.RGB, B=@n129.RGB), Alpha=@n128.G)) {name=UVs_Logo_A_Offset, nodeColor=<rgba 0 0.625 1 1>}
  n130 = VectorParameter() {group=Wrap_NeonGraffitiLava_UVs, parameterName=NeonGL_LavaBox_Offsest_Weapon, defaultValue=<rgba -0.75 0.625 0 1>}
  n131 = VectorParameter() {group=Wrap_NeonGraffitiLava_UVs, parameterName=NeonGL_LavaBox_Offsest_Vehicle, defaultValue=<rgba -0.5 0.5 0 1>}
  n132 = VectorParameter() {group=Wrap_NeonGraffitiLava_UVs, parameterName=NeonGL_LavaBox_Offsest_VehicleSide, defaultValue=<rgba -0.5 0.5 0 1>}
  n133 = MaterialFunctionCall(Float4=NamedRerouteUsage()) {materialFunction=&BreakOutFloat4Components}
  n134 = VectorParameter() {group=Wrap_NeonGraffitiLava_UVs, parameterName=NeonGL_LavaBox_Offsest_Mech, defaultValue=<rgba -0.5 0.5 0 1>}
  n135 = VectorParameter() {group=Wrap_NeonGraffitiLava_UVs, parameterName=NeonGL_LavaBox_Offsest_Character, defaultValue=<rgba -0.5 0.5 0 1>}
  n136 = LinearInterpolate(
   A=LinearInterpolate(A=@n130.RGB, B=LinearInterpolate(A=@n131.RGB, B=@n132.RGB, Alpha=@n133.G), Alpha=@n133.R)
   B=LinearInterpolate(A=@n134.RGB, B=@n135.RGB, Alpha=@n133.A)
   Alpha=@n133.B
  )
  n137 = VectorParameter() {group=Wrap_NeonGraffitiLava_UVs, parameterName=NeonGL_LavaBox_Offsest_Backpack, defaultValue=<rgba -0.5 0.5 0 1>}
  n138 = MaterialFunctionCall(Float2=NamedRerouteUsage()) {materialFunction=&BreakOutFloat2Components}
  n139 = VectorParameter() {group=Wrap_NeonGraffitiLava_UVs, parameterName=NeonGL_LavaBox_Offsest_Pickaxe, defaultValue=<rgba -0.5 0.5 0 1>}
  NamedRerouteDeclaration(None=LinearInterpolate(A=LinearInterpolate(A=@n136, B=@n137.RGB, Alpha=@n138.R), B=@n139.RGB, Alpha=@n138.G)) {name=UVs_Lava_Offset, nodeColor=<rgba 0 0.625 1 1>}
  n140 = ScalarParameter() {group=Wrap_NeonGraffitiLava_UVs, parameterName=NeonGL_Logo_Rotation_Vehicle, sortPriority=150}
  n141 = ScalarParameter() {group=Wrap_NeonGraffitiLava_UVs, parameterName=NeonGL_Logo_Rotation_Vehicle_Side, sortPriority=150}
  n142 = MaterialFunctionCall(Float4=NamedRerouteUsage()) {materialFunction=&BreakOutFloat4Components}
  n143 = LinearInterpolate(
   A=LinearInterpolate(
    A=ScalarParameter() {group=Wrap_NeonGraffitiLava_UVs, parameterName=NeonGL_Logo_Rotation_Weapon, sortPriority=150}
    B=LinearInterpolate(A=@n140, B=@n141, Alpha=@n142.G)
    Alpha=@n142.R
   )
   B=LinearInterpolate(
    A=ScalarParameter() {group=Wrap_NeonGraffitiLava_UVs, parameterName=NeonGL_Logo_Rotation_Mech, sortPriority=150}
    B=ScalarParameter() {group=Wrap_NeonGraffitiLava_UVs, parameterName=NeonGL_Logo_Rotation_Character, sortPriority=150}
    Alpha=@n142.A
   )
   Alpha=@n142.B
  )
  n144 = ScalarParameter() {group=Wrap_NeonGraffitiLava_UVs, parameterName=NeonGL_Logo_Rotation_Backpack, sortPriority=150}
  n145 = MaterialFunctionCall(Float2=NamedRerouteUsage()) {materialFunction=&BreakOutFloat2Components}
  NamedRerouteDeclaration(None=LinearInterpolate(
    A=LinearInterpolate(A=@n143, B=@n144, Alpha=@n145.R)
    B=ScalarParameter() {group=Wrap_NeonGraffitiLava_UVs, parameterName=NeonGL_Logo_Rotation_Pickaxe, sortPriority=150}
    Alpha=@n145.G
   )) {name=UVs_Logo_Rotation, nodeColor=<rgba 0 0.7188 1 1>}
  n146 = VectorParameter() {group=Wrap_NeonGraffitiLava_UVs, parameterName=NeonGL_LogoB_Offsest_Weapon, sortPriority=100, defaultValue=<rgba 0.3 -0.025 0 0.2>}
  n147 = VectorParameter() {group=Wrap_NeonGraffitiLava_UVs, parameterName=NeonGL_LogoB_Offsest_Vehicle, sortPriority=100, defaultValue=<rgba 0 0 0 1>}
  n148 = AppendVector(A=@n147.RGB, B=@n147.RGB)
  n149 = VectorParameter() {group=Wrap_NeonGraffitiLava_UVs, parameterName=NeonGL_LogoB_Offsest_VehicleSide, sortPriority=100, defaultValue=<rgba 0 0 0 1>}
  n150 = AppendVector(A=@n149.RGB, B=@n149.RGB)
  n151 = MaterialFunctionCall(Float4=NamedRerouteUsage()) {materialFunction=&BreakOutFloat4Components}
  n152 = VectorParameter() {group=Wrap_NeonGraffitiLava_UVs, parameterName=NeonGL_LogoB_Offsest_Mech, sortPriority=100, defaultValue=<rgba 0 0 0 1>}
  n153 = VectorParameter() {group=Wrap_NeonGraffitiLava_UVs, parameterName=NeonGL_LogoB_Offsest_Character, sortPriority=100, defaultValue=<rgba 0 0 0 1>}
  n154 = LinearInterpolate(
   A=LinearInterpolate(A=AppendVector(A=@n146.RGB, B=@n146.RGB), B=LinearInterpolate(A=@n148, B=@n150, Alpha=@n151.G), Alpha=@n151.R)
   B=LinearInterpolate(A=AppendVector(A=@n152.RGB, B=@n152.RGB), B=AppendVector(A=@n153.RGB, B=@n153.RGB), Alpha=@n151.A)
   Alpha=@n151.B
  )
  n155 = VectorParameter() {group=Wrap_NeonGraffitiLava_UVs, parameterName=NeonGL_LogoB_Offsest_Backpack, sortPriority=100, defaultValue=<rgba 0 0 0 1>}
  n156 = AppendVector(A=@n155.RGB, B=@n155.RGB)
  n157 = MaterialFunctionCall(Float2=NamedRerouteUsage()) {materialFunction=&BreakOutFloat2Components}
  n158 = VectorParameter() {group=Wrap_NeonGraffitiLava_UVs, parameterName=NeonGL_LogoB_Offsest_Pickaxe, sortPriority=100, defaultValue=<rgba 0 0 0 1>}
  NamedRerouteDeclaration(None=LinearInterpolate(A=LinearInterpolate(A=@n154, B=@n156, Alpha=@n157.R), B=AppendVector(A=@n158.RGB, B=@n158.RGB), Alpha=@n157.G)) {name=UVs_Logo_B_Offset, nodeColor=<rgba 0 0.625 1 1>}
  n159 = Multiply(A=NamedRerouteUsage(), B=ScalarParameter() {defaultValue=1, group=Wrap_NeonGraffitiLava, parameterName=NeonGL_Lava_Noise_Bck_TileRate})
  NamedRerouteDeclaration(None=TextureSample(UVs=Panner(Coordinate=@n159, Speed=@n36), Tex=@n37).R) {name=Msk_Noise_Bck, nodeColor=<rgba 0 1 0.6016 1>}
  n160 = CurveAtlasRowParameter(CurveTime=@n52) {
   atlas=&CruveAtlas_Wraps_01
   curve=/Game/Weapons/WeaponSkins/Wraps/Textures/0_ColorCurves/CC_NGL_Colors
   defaultValue=74
   group=Wrap_NeonGraffitiLava
   parameterName=NeonGL_BG_ColorCurve_Colors
  }
  NamedRerouteDeclaration(None=MaterialFunctionCall(
    Index=ScalarParameter() {defaultValue=74, group=Wrap_NeonGraffitiLava, parameterName="NeonGL_BG_ColorCurve_Colors  Index"}
    CCinput=@n160
    UV=@n52
   ) {materialFunction=&MF_Global_CC_Tex}.3V) {name=Lava_edge_Glow, nodeColor=<rgba 0 1 0.5078 1>}
  n161 = MaterialFunctionCall(Float3=Abs(None=NamedRerouteUsage())) {materialFunction=&BreakOutFloat3Components}
  n162 = Constant() {r=0.5}
  n163 = VectorParameter() {group=Wrap_NeonGraffitiLava_Embers, parameterName=NeonGL_Embers_Speed_A, defaultValue=<rgba -0.013 0.013 -0.05 1>}
  n164 = Add(
   A=Multiply(A=NamedRerouteUsage(), B=ScalarParameter() {defaultValue=2, group=Wrap_NeonGraffitiLava_Embers, parameterName=NeonGL_Embers_TileRate_A})
   B=Multiply(A=Time(), B=@n163.RGB)
  )
  n165 = ComponentMask(None=@n164) {b=true, desc=YZ, g=true}
  n166 = ComponentMask(None=@n164) {b=true, desc=XZ, r=true}
  NamedRerouteDeclaration(None=If(
    A=@n161.B
    B=@n162
    "A > B"=ComponentMask(None=@n164) {desc=YX, g=true, r=true}
    "A < B"=If(A=@n161.R, B=@n162, "A > B"=@n165, "A < B"=@n166)
   )) {name=UVS_Cheap_Tri, nodeColor=<rgba 1 0 0.8438 1>}
  n167 = VectorParameter() {group=Wrap_NeonGraffitiLava_UVs, parameterName=NeonGL_Lava_Drip_Weapon, defaultValue=<rgba 3 2 0 1>}
  n168 = VectorParameter() {group=Wrap_NeonGraffitiLava_UVs, parameterName=NeonGL_Lava_Drip_Vehicle, defaultValue=<rgba 1 1 1 1>}
  n169 = Multiply(A=@n168.RGB, B=@n168.RGB)
  n170 = VectorParameter() {group=Wrap_NeonGraffitiLava_UVs, parameterName=NeonGL_Lava_Drip_VehicleSide, defaultValue=<rgba 1 1 1 1>}
  n171 = Multiply(A=@n170.RGB, B=@n170.RGB)
  n172 = MaterialFunctionCall(Float4=NamedRerouteUsage()) {materialFunction=&BreakOutFloat4Components}
  n173 = VectorParameter() {group=Wrap_NeonGraffitiLava_UVs, parameterName=NeonGL_Lava_Drip_Mech, defaultValue=<rgba 1 1 1 1>}
  n174 = VectorParameter() {group=Wrap_NeonGraffitiLava_UVs, parameterName=NeonGL_Lava_Drip_Character, defaultValue=<rgba 1 1 1 1>}
  n175 = LinearInterpolate(
   A=LinearInterpolate(A=Multiply(A=@n167.RGB, B=@n167.RGB), B=LinearInterpolate(A=@n169, B=@n171, Alpha=@n172.G), Alpha=@n172.R)
   B=LinearInterpolate(A=Multiply(A=@n173.RGB, B=@n173.RGB), B=Multiply(A=@n174.RGB, B=@n174.RGB), Alpha=@n172.A)
   Alpha=@n172.B
  )
  n176 = VectorParameter() {group=Wrap_NeonGraffitiLava_UVs, parameterName=NeonGL_Lava_Drip_Backpack, defaultValue=<rgba 1 1 1 1>}
  n177 = Multiply(A=@n176.RGB, B=@n176.RGB)
  n178 = MaterialFunctionCall(Float2=NamedRerouteUsage()) {materialFunction=&BreakOutFloat2Components}
  n179 = VectorParameter() {group=Wrap_NeonGraffitiLava_UVs, parameterName=NeonGL_Lava_Drip_Pickaxe, defaultValue=<rgba 1 1 1 1>}
  NamedRerouteDeclaration(None=LinearInterpolate(A=LinearInterpolate(A=@n175, B=@n177, Alpha=@n178.R), B=Multiply(A=@n179.RGB, B=@n179.RGB), Alpha=@n178.G)) {name=UVs_Drip_Scale, nodeColor=<rgba 0 0.625 1 1>}
  n180 = ScalarParameter() {defaultValue=1, group=Wrap_NeonGraffitiLava_UVs, parameterName=NeonGL_Scale_Noise_Vehicle, sortPriority=150}
  n181 = ScalarParameter() {defaultValue=1, group=Wrap_NeonGraffitiLava_UVs, parameterName=NeonGL_Scale_Noise_Vehicle_Side, sortPriority=150}
  n182 = MaterialFunctionCall(Float4=NamedRerouteUsage()) {materialFunction=&BreakOutFloat4Components}
  n183 = LinearInterpolate(
   A=LinearInterpolate(
    A=ScalarParameter() {defaultValue=3, group=Wrap_NeonGraffitiLava_UVs, parameterName=NeonGL_Scale_Noise_Weapon, sortPriority=150}
    B=LinearInterpolate(A=@n180, B=@n181, Alpha=@n182.G)
    Alpha=@n182.R
   )
   B=LinearInterpolate(
    A=ScalarParameter() {defaultValue=1, group=Wrap_NeonGraffitiLava_UVs, parameterName=NeonGL_Scale_Noise_Mech, sortPriority=150}
    B=ScalarParameter() {defaultValue=1, group=Wrap_NeonGraffitiLava_UVs, parameterName=NeonGL_Scale_Noise_Character, sortPriority=150}
    Alpha=@n182.A
   )
   Alpha=@n182.B
  )
  n184 = ScalarParameter() {defaultValue=1, group=Wrap_NeonGraffitiLava_UVs, parameterName=NeonGL_Scale_Noise_Backpack, sortPriority=150}
  n185 = MaterialFunctionCall(Float2=NamedRerouteUsage()) {materialFunction=&BreakOutFloat2Components}
  NamedRerouteDeclaration(None=LinearInterpolate(
    A=LinearInterpolate(A=@n183, B=@n184, Alpha=@n185.R)
    B=ScalarParameter() {defaultValue=1, group=Wrap_NeonGraffitiLava_UVs, parameterName=NeonGL_Scale_Noise_Pickaxe, sortPriority=150}
    Alpha=@n185.G
   )) {name=UVs_Noise_Scale, nodeColor=<rgba 0 0.7188 1 1>}
  n186 = ScalarParameter() {group=Wrap_NeonGraffitiLava_UVs, parameterName=NeonGL_Glow_Rotation_Vehicle}
  n187 = ScalarParameter() {group=Wrap_NeonGraffitiLava_UVs, parameterName=NeonGL_Glow_Rotation_Vehicle_Side}
  n188 = MaterialFunctionCall(Float4=NamedRerouteUsage()) {materialFunction=&BreakOutFloat4Components}
  n189 = LinearInterpolate(
   A=LinearInterpolate(
    A=ScalarParameter() {group=Wrap_NeonGraffitiLava_UVs, parameterName=NeonGL_Glow_Rotation_Weapon}
    B=LinearInterpolate(A=@n186, B=@n187, Alpha=@n188.G)
    Alpha=@n188.R
   )
   B=LinearInterpolate(
    A=ScalarParameter() {group=Wrap_NeonGraffitiLava_UVs, parameterName=NeonGL_Glow_Rotation_Mech}
    B=ScalarParameter() {group=Wrap_NeonGraffitiLava_UVs, parameterName=NeonGL_Glow_Rotation_Character}
    Alpha=@n188.A
   )
   Alpha=@n188.B
  )
  n190 = ScalarParameter() {group=Wrap_NeonGraffitiLava_UVs, parameterName=NeonGL_Glow_Rotation_Backpack}
  n191 = MaterialFunctionCall(Float2=NamedRerouteUsage()) {materialFunction=&BreakOutFloat2Components}
  NamedRerouteDeclaration(None=LinearInterpolate(
    A=LinearInterpolate(A=@n189, B=@n190, Alpha=@n191.R)
    B=ScalarParameter() {group=Wrap_NeonGraffitiLava_UVs, parameterName=NeonGL_Glow_Rotation_Pickaxe}
    Alpha=@n191.G
   )) {name=UVs_Glow_Rotation, nodeColor=<rgba 0 0.7188 1 1>}
  n192 = VectorParameter() {group=Wrap_NeonGraffitiLava_UVs, parameterName=NeonGL_LavaBox_Scale_Weapon, defaultValue=<rgba 0.65 0.13 0 1>}
  n193 = VectorParameter() {group=Wrap_NeonGraffitiLava_UVs, parameterName=NeonGL_LavaBox_Scale_Vehicle, defaultValue=<rgba 0.1 0.1 0 1>}
  n194 = VectorParameter() {group=Wrap_NeonGraffitiLava_UVs, parameterName=NeonGL_LavaBox_Scale_VehicleSide, defaultValue=<rgba 0.1 0.1 0 1>}
  n195 = MaterialFunctionCall(Float4=NamedRerouteUsage()) {materialFunction=&BreakOutFloat4Components}
  n196 = VectorParameter() {group=Wrap_NeonGraffitiLava_UVs, parameterName=NeonGL_LavaBox_Scale_Mech, defaultValue=<rgba 0.1 0.1 0 1>}
  n197 = VectorParameter() {group=Wrap_NeonGraffitiLava_UVs, parameterName=NeonGL_LavaBox_Scale_Character, defaultValue=<rgba 0.1 0.1 0 1>}
  n198 = LinearInterpolate(
   A=LinearInterpolate(A=@n192.RGB, B=LinearInterpolate(A=@n193.RGB, B=@n194.RGB, Alpha=@n195.G), Alpha=@n195.R)
   B=LinearInterpolate(A=@n196.RGB, B=@n197.RGB, Alpha=@n195.A)
   Alpha=@n195.B
  )
  n199 = VectorParameter() {group=Wrap_NeonGraffitiLava_UVs, parameterName=NeonGL_LavaBox_Scale_Backpack, defaultValue=<rgba 0.1 0.1 0 1>}
  n200 = MaterialFunctionCall(Float2=NamedRerouteUsage()) {materialFunction=&BreakOutFloat2Components}
  n201 = VectorParameter() {group=Wrap_NeonGraffitiLava_UVs, parameterName=NeonGL_LavaBox_Scale_Pickaxe, defaultValue=<rgba 0.1 0.1 0 1>}
  NamedRerouteDeclaration(None=LinearInterpolate(A=LinearInterpolate(A=@n198, B=@n199.RGB, Alpha=@n200.R), B=@n201.RGB, Alpha=@n200.G)) {name=UVs_Lava_Scale, nodeColor=<rgba 0 0.625 1 1>}
  n202 = FunctionInput(Preview=Constant4Vector().RGBA) {bUsePreviewValueAsDefault=true, inputName=IsState, inputType=FunctionInput_Vector4, sortPriority=4}
  NamedRerouteDeclaration(None=MaterialFunctionCall(Float4=NamedRerouteDeclaration(None=@n202) {name=IsState, nodeColor=<rgba 0.1497 1 0.1633 1>}) {materialFunction=&BreakOutFloat4Components}.A) {name=IsCharacter, nodeColor=<rgba 1 0.9141 0 1>}
  n203 = FunctionInput(Preview=Constant2Vector()) {bUsePreviewValueAsDefault=true, inputName=IsState_B, inputType=FunctionInput_Vector2, sortPriority=4}
  NamedRerouteDeclaration(None=MaterialFunctionCall(Float2=NamedRerouteDeclaration(None=@n203) {name=IsState_B, nodeColor=<rgba 0.1497 1 0.1633 1>}) {materialFunction=&BreakOutFloat2Components}.R) {name=IsBackPack, nodeColor=<rgba 1 0.9141 0 1>}
  n204 = TextureObjectParameter() {group=Wrap_NeonGraffitiLava, parameterName=NeonGL__Tex_Msk2, sortPriority=10, texture=&T_BlacksRGB}
  NamedRerouteDeclaration(None=@n204) {name=Tex_Msk2, nodeColor=<rgba 1 0.5859 0 1>}
  NamedRerouteDeclaration(None=@n4) {name=Embers, nodeColor=<rgba 0 0.7656 1 1>}
  n205 = NamedRerouteDeclaration(None=@MaterialExpressionMaterialFunctionCall_8.Alpha) {name=Lava_Msk, nodeColor=<rgba 0.6406 0 1 1>}
  NamedRerouteDeclaration(None=Multiply(A=OneMinus(None=@n205), B=@n52)) {name=Lava_Glow, nodeColor=<rgba 0.6406 0 1 1>}
  n206 = PreSkinnedLocalBounds()
  n207 = Subtract(A=VertexInterpolator(VS=PreSkinnedPosition()).PS, B=@n206.Min)
  n208 = MaterialFunctionCall(Color=Subtract(A=@n206.Max, B=@n206.Max)) {materialFunction=&SplitComponents}
  n209 = Max(A=Max(A=@n208.R, B=@n208.R), B=@n208.B)
  NamedRerouteDeclaration(None=FunctionInput(Preview=Divide(A=@n207, B=@n209)) {bUsePreviewValueAsDefault=true, inputName=Calcated_Bounds, sortPriority=11}) {name=Caclcated_Bounds, nodeColor=<rgba 0 0.9297 1 1>}
  NamedRerouteDeclaration(None=@n52) {name=Lava, nodeColor=<rgba 0.6406 0 1 1>}
  n210 = PreSkinnedNormal()
  NamedRerouteDeclaration(None=FunctionInput(Preview=VertexInterpolator(VS=@n210).PS) {bUsePreviewValueAsDefault=true, inputName=Pre-Skinned Local Normal, sortPriority=10}) {name=PreSkin_Normal, nodeColor=<rgba 0.8281 1 0 1>}
  NamedRerouteDeclaration(None=@n64.Normal) {name=Root_Normal, nodeColor=<rgba 0.4062 0 1 1>}
  n211 = VectorParameter() {group=Wrap_NeonGraffitiLava_Embers, parameterName=NeonGL_Embers_Speed_A, defaultValue=<rgba -0.013 0.013 -0.05 1>}
  Panner(
   Coordinate=Multiply(A=@n2, B=ScalarParameter() {defaultValue=2, group=Wrap_NeonGraffitiLava_Embers, parameterName=NeonGL_Embers_TileRate_A})
   Speed=ComponentMask(None=@n211.RGB) {g=true, r=true}
  )
  ScalarParameter() {defaultValue=3, group=Wrap_NeonGraffitiLava, parameterName=NeonGL_Lava_Noise_TileRate}
  n212 = GetMaterialAttributes(None=@n61) {attributeGetTypes [#3]}
  n213 = Add(A=OneMinus(None=@n93), B=ScalarParameter() {group=Wrap_NeonGraffitiLava, parameterName=NeonGL_Roughness_edge, sortPriority=5})
  SetMaterialAttributes!(MaterialAttributes=@n212.MaterialAttributes, Roughness=Multiply(A=@n212.Roughness, B=Saturate(None=@n213))) {attributeSetTypes [#3]}
  SphereMask(A=@n23.UVs, B=Constant() {r=0.5}, Hardness=Constant() {r=1}) {attenuationRadius=0.5, hardnessPercent=1}
  VertexNormalWS()
 ]}
