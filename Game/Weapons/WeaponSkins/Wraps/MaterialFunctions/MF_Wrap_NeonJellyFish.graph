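# Material function for the Neon Jellyfish wrap on weapons, vehicles and mechs, blended over incoming material attributes by customization mask behind a static switch.
# Scrolling top/side projections warp a second projected mask, with per-type UV settings.
# Those masks index a curve-atlas gradient for emissive color and gate two warped, scrolling glow layers over a dark base.
# A Blueprint-set timestamp drives an eased offset that shifts warp and glow UVs.
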MaterialFunction /Game/Weapons/WeaponSkins/Wraps/MaterialFunctions/MF_Wrap_NeonJellyFish
graph {nodes [
  Constant()
  n1 = VectorParameter() {group=Wrap_NeonJellyfish_Uvs, parameterName=NJF_Speed_Warp_Weapon_03B, sortPriority=520, defaultValue=<rgba 0 0.1 0 1>}
  n2 = VectorParameter() {group=Wrap_NeonJellyfish_Uvs, parameterName=NJF_Speed_Warp_Vehicle_03B, sortPriority=520, defaultValue=<rgba 0 0 0 1>}
  n3 = FunctionInput(Preview=Constant()) {inputName=isVehicle, inputType=FunctionInput_Scalar, sortPriority=5}
  n4 = LinearInterpolate(A=@n1.RGB, B=@n2.RGB, Alpha=@n3)
  n5 = VectorParameter() {group=Wrap_NeonJellyfish_Uvs, parameterName=NJF_Speed_Warp_Mech_03B, sortPriority=520, defaultValue=<rgba 1 1 1 1>}
  n6 = Constant()
  n7 = FunctionInput(Preview=@n6) {bUsePreviewValueAsDefault=true, inputName=isMech, inputType=FunctionInput_Scalar, sortPriority=4}
  n8 = Constant()
  n9 = Subtract(A=Time(), B=ScalarParameter() {defaultValue=-9999999, group=Wrap_NeonJellyfish, parameterName=TimeFrom-BP----------------, sortPriority=600})
  n10 = ScalarParameter() {defaultValue=0.8, group=Wrap_NeonJellyfish, parameterName=NJF_Time_Offset_Duration, sortPriority=601}
  n11 = Power(
   Base=Saturate(None=Divide(A=@n9, B=@n10))
   Exp=ScalarParameter() {defaultValue=0.5, group=Wrap_NeonJellyfish, parameterName=NJF_Time_Cuve, sortPriority=602}
  )
  n12 = LinearInterpolate(B=ScalarParameter() {group=Wrap_NeonJellyfish, parameterName=NJF_Time_SpeedAmount, sortPriority=603}, Alpha=@n11) {constB=0.3}
  n13 = LinearInterpolate(
   A=LinearInterpolate(
    A=MaterialFunctionCall(R=@n8, G=@n12, B=@n8) {materialFunction=&AppendMany}.RGB
    B=MaterialFunctionCall(R=@n12, G=@n8, B=@n8) {materialFunction=&AppendMany}.RGB
    Alpha=@n3
   )
   B=MaterialFunctionCall(R=@n8, G=@n8, B=@n12) {materialFunction=&AppendMany}.RGB
   Alpha=@n7
  )
  n14 = Add(A=Multiply(A=Time(), B=LinearInterpolate(A=@n4, B=@n5.RGB, Alpha=@n7)), B=@n13)
  Frac(None=@n14)
  n15 = VectorParameter() {group=Wrap_NeonJellyfish_Uvs, parameterName=NJF_Speed_Warp_Weapon_03A, sortPriority=520, defaultValue=<rgba 0 0.1 0 1>}
  n16 = VectorParameter() {group=Wrap_NeonJellyfish_Uvs, parameterName=NJF_Speed_Warp_Vehicle_03A, sortPriority=520, defaultValue=<rgba 0 0 0 1>}
  n17 = LinearInterpolate(A=@n15.RGB, B=@n16.RGB, Alpha=@n3)
  n18 = VectorParameter() {group=Wrap_NeonJellyfish_Uvs, parameterName=NJF_Speed_Warp_Mech_03A, sortPriority=520, defaultValue=<rgba 1 1 1 1>}
  n19 = Add(A=Multiply(A=Time(), B=LinearInterpolate(A=@n17, B=@n18.RGB, Alpha=@n7)), B=@n13)
  Frac(None=@n19)
  n20 = GetMaterialAttributes(None=FunctionInput() {inputName=Material, inputType=FunctionInput_MaterialAttributes}) {attributeGetTypes [#1, #2, #3]}
  n21 = VectorParameter() {group=Wrap_NeonJellyfish, parameterName=NJF_BaseColor, sortPriority=0, defaultValue=<rgba 0.015 0.015 0.015 1>}
  n22 = VectorParameter() {group=Wrap_NeonJellyfish, parameterName=NJF_Fill_alternate_Axis_Color 02, sortPriority=51}
  n23 = AppendVector(A=@n22.RGB, B=@n22.RGB)
  n24 = VectorParameter() {group=Wrap_NeonJellyfish_Uvs, parameterName=NJF-Rotation-Weapon-01---------------------, sortPriority=150, defaultValue=<rgba 0 0 -0.25 1>}
  n25 = VectorParameter() {group=Wrap_NeonJellyfish_Uvs, parameterName=NJF_Rotation_Vehicle_01, sortPriority=151, defaultValue=<rgba 0 0 0 1>}
  n26 = VectorParameter() {group=Wrap_NeonJellyfish_Uvs, parameterName=NJF_Rotation_Vehicle_Side_01, sortPriority=151, defaultValue=<rgba 0 0 0 1>}
  n27 = FunctionInput(Preview=@n6) {bUsePreviewValueAsDefault=true, inputName=Side Project Vehicle, inputType=FunctionInput_Scalar}
  n28 = VectorParameter() {group=Wrap_NeonJellyfish_Uvs, parameterName=NJF_Rotation_Mech_01, sortPriority=153, defaultValue=<rgba 0.25 0.25 0 1>}
  n29 = LinearInterpolate(A=LinearInterpolate(A=@n24.RGB, B=LinearInterpolate(A=@n25.RGB, B=@n26.RGB, Alpha=@n27), Alpha=@n3), B=@n28.RGB, Alpha=@n7)
  n30 = ScalarParameter() {group=Wrap_NeonJellyfish, parameterName=NJF_Fill_alternate_Axis_Opacity 01/02, sortPriority=52}
  n31 = ScalarParameter() {defaultValue=1, group=Wrap_NeonJellyfish, parameterName=NJF_Blend_Contrast, sortPriority=-1}
  n32 = TextureObjectParameter() {group=Wrap_NeonJellyfish, parameterName=NJF_Mask_02, sortPriority=50, texture=&T_BlacksRGB}
  n33 = TextureObjectParameter() {group=Wrap_NeonJellyfish, parameterName=NJF_Mask_03, sortPriority=210, texture=&T_BlacksRGB}
  n34 = ScalarParameter() {group=Wrap_NeonJellyfish, parameterName=NJF_Fill_alternate_Axis_Opacity 03, sortPriority=211}
  n35 = VectorParameter() {group=Wrap_NeonJellyfish, parameterName=NJF_Fill_alternate_Axis_Color 03, sortPriority=211}
  n36 = AppendVector(A=@n35.RGB, B=@n35.RGB)
  n37 = FunctionInput(Preview=VertexInterpolator(VS=PreSkinnedNormal()).PS) {bUsePreviewValueAsDefault=true, inputName=Pre-Skinned Local Normal, sortPriority=10}
  n38 = ScalarParameter() {defaultValue=1, group=Wrap_NeonJellyfish_Uvs, parameterName=NJF_Scale_Warp_Weapon_03, sortPriority=500}
  n39 = ScalarParameter() {defaultValue=1, group=Wrap_NeonJellyfish_Uvs, parameterName=NJF_Scale_Warp_Vehicle_03, sortPriority=500}
  n40 = VertexInterpolator(VS=PreSkinnedPosition())
  n41 = PreSkinnedLocalBounds()
  n42 = FunctionInput(Preview=Divide(A=Subtract(A=@n40.PS, B=@n41.Min), B=Subtract(A=@n41.Max, B=@n41.Max))) {bUsePreviewValueAsDefault=true, inputName=Calcated_Bounds, sortPriority=11}
  n43 = VectorParameter() {group=Wrap_NeonJellyfish_Uvs, parameterName=NJF_Offset_Weapon_01---------------------, sortPriority=160, defaultValue=<rgba 0 0 -1 1>}
  n44 = VectorParameter() {group=Wrap_NeonJellyfish_Uvs, parameterName=NJF_Offset_Vehicle_Side_01, sortPriority=161, defaultValue=<rgba 0 0 0 1>}
  n45 = LinearInterpolate(A=VectorParameter() {group=Wrap_NeonJellyfish_Uvs, parameterName=NJF_Offset_Vehicle_01, sortPriority=161}.RGB, B=@n44.RGB, Alpha=@n27)
  n46 = VectorParameter() {group=Wrap_NeonJellyfish_Uvs, parameterName=NJF_Offset_Mech_01, sortPriority=162, defaultValue=<rgba 0 0 0 1>}
  n47 = Add(A=@n42, B=LinearInterpolate(A=LinearInterpolate(A=@n43.RGB, B=@n45, Alpha=@n3), B=@n46.RGB, Alpha=@n7))
  n48 = Multiply(
   A=@n47
   B=LinearInterpolate(
    A=LinearInterpolate(A=@n38, B=@n39, Alpha=@n3)
    B=ScalarParameter() {defaultValue=1, group=Wrap_NeonJellyfish_Uvs, parameterName=NJF_Scale_Warp_Mech_03, sortPriority=500}
    Alpha=@n7
   )
  )
  n49 = MaterialFunctionCall!(
   IsCharacter
   IsStatsB
   Texture=@n33
   Rotation_Axis=@n29
   Caculated_Bounds=Add(A=@n48, B=@n19)
   Pre_Skinned_Normals=@n37
   Blend_Contrast=@n31
   Fill_alternate_Axis_Opacity=@n34
   "Fill_alternate_Axis_Color_A=Roughness"=@n36
   Is_Vehicle=@n3
   Is_Mech=@n7
   Enable_Fill_Z_Weapon
   Fill_Z_Weapon_opacity
   Fill_Z_Weapon_Color
  ) {materialFunction=&/MF_TopsSideProjection_Base}
  n50 = Multiply(
   A=ScalarParameter() {defaultValue=0.5, group=Wrap_NeonJellyfish, parameterName=NJF_Mask_03A_Warp_Amount, sortPriority=211}
   B=Add(A=ComponentMask(None=@n49.RGB) {g=true}) {constB=-0.25}
  )
  n51 = ScalarParameter() {group=Wrap_NeonJellyfish, parameterName=NJF_WarpGradent_03_Mask_Offset_Weapon, sliderMin=-1, sortPriority=250}
  n52 = ScalarParameter() {defaultValue=-1, group=Wrap_NeonJellyfish, parameterName=NJF_WarpGradent_03_Mask_Offset_Vehicle, sortPriority=250}
  n53 = ScalarParameter() {defaultValue=-1, group=Wrap_NeonJellyfish, parameterName=NJF_WarpGradent_03_Mask_Falloff_Weapon, sortPriority=250}
  n54 = ScalarParameter() {defaultValue=-1, group=Wrap_NeonJellyfish, parameterName=NJF_WarpGradent_03_Mask_Falloff_Vehicle, sortPriority=250}
  n55 = LinearInterpolate(
   A=FunctionInput(Preview=Constant3Vector() {constant=<rgba 0 1 0 0>}) {bUsePreviewValueAsDefault=true, inputName=ForwardAxis}
   B=Constant3Vector() {constant=<rgba 0 1 0 0>}
   Alpha=ScalarParameter() {group=Wrap_NeonJellyfish, parameterName=NJF_IS_Roll, sortPriority=-1}
  )
  n56 = MaterialFunctionCall!(
   Bounds=@n42
   ForwardAxis=@n55
   Offset=LinearInterpolate(
    A=LinearInterpolate(A=@n51, B=@n52, Alpha=@n3)
    B=ScalarParameter() {defaultValue=-1, group=Wrap_NeonJellyfish, parameterName=NJF_WarpGradent_03_Mask_Offset_Mech, sortPriority=250}
    Alpha=@n7
   )
   FallOff=LinearInterpolate(
    A=LinearInterpolate(A=@n53, B=@n54, Alpha=@n3)
    B=ScalarParameter() {defaultValue=-1, group=Wrap_NeonJellyfish, parameterName=NJF_WarpGradent_03_Mask_Falloff_Mech, sortPriority=250}
    Alpha=@n7
   )
  ) {materialFunction=&/MF_Wrap_Gradent_Simple}
  n57 = VectorParameter() {group=Wrap_NeonJellyfish_Uvs, parameterName="NJF_Scale_Weapon_01 A=uniform", sortPriority=199, defaultValue=<rgba 1 1 1 1>}
  n58 = Multiply(A=@n57.RGB, B=@n57.RGB)
  n59 = VectorParameter() {group=Wrap_NeonJellyfish_Uvs, parameterName="NJF_Scale_Vehicle_01 A=uniform", sortPriority=200, defaultValue=<rgba 1 1 1 1>}
  n60 = VectorParameter() {group=Wrap_NeonJellyfish_Uvs, parameterName="NJF_Scale_Vehicle_Side_01 A=uniform", sortPriority=202, defaultValue=<rgba 1 1 1 1>}
  n61 = LinearInterpolate(A=Multiply(A=@n59.RGB, B=@n59.RGB), B=Multiply(A=@n60.RGB, B=@n60.RGB), Alpha=@n27)
  n62 = VectorParameter() {group=Wrap_NeonJellyfish_Uvs, parameterName="NJF_Scale_Mech_01 A=uniform", sortPriority=201, defaultValue=<rgba 1 1 1 1>}
  n63 = Multiply(A=@n47, B=LinearInterpolate(A=LinearInterpolate(A=@n58, B=@n61, Alpha=@n3), B=Multiply(A=@n62.RGB, B=@n62.RGB), Alpha=@n7))
  n64 = MaterialFunctionCall(
   Texture=@n32
   Rotation_Axis=@n29
   Caculated_Bounds=Add(A=@n63, B=Multiply(A=@n50, B=@n56.Raw_Data))
   Pre_Skinned_Normals=@n37
   Blend_Contrast=@n31
   Fill_alternate_Axis_Opacity=@n30
   "Fill_alternate_Axis_Color_A=Roughness"=@n23
   Is_Vehicle=@n3
   Is_Mech=@n7
  ) {materialFunction=&MF_TopsSideProjection_Base}
  n65 = ComponentMask(None=@n64.Combined) {a=true}
  n66 = Power(Base=Multiply(A=@n65) {constB=10}) {constExponent=5}
  n67 = Constant() {r=0.1}
  n68 = VectorParameter() {group=Wrap_NeonJellyfish, parameterName=NJF_Glow_Spread_2, sortPriority=420, defaultValue=<rgba 1 1 0 1>}
  n69 = VectorParameter() {group=Wrap_NeonJellyfish, parameterName=NJF_Glow_Speed_2_Weapon, sortPriority=430, defaultValue=<rgba 0.5 0 0 1>}
  n70 = VectorParameter() {group=Wrap_NeonJellyfish, parameterName=NJF_Glow_Speed_2_Vehicle, sortPriority=430, defaultValue=<rgba 0 0 0 1>}
  n71 = LinearInterpolate(A=@n69.RGB, B=@n70.RGB, Alpha=@n3)
  n72 = Multiply(
   A=ScalarParameter() {defaultValue=0.5, group=Wrap_NeonJellyfish, parameterName=NJF_Glow_Warp_2_uvs, sortPriority=435}
   B=ComponentMask(None=@n64.RGB) {b=true}
  )
  n73 = ScalarParameter() {defaultValue=1, group=Wrap_NeonJellyfish, parameterName=NJF_Glow_Rotate_Vehicle 1/2, sortPriority=399}
  n74 = AppendVector(
   A=LinearInterpolate(B=ScalarParameter() {defaultValue=0.3, group=Wrap_NeonJellyfish, parameterName=NJF_Time_Glow_SpeedAmount, sortPriority=603}, Alpha=@n11) {constB=0.5}
   B=Constant()
  )
  n75 = TextureObjectParameter() {group=Wrap_NeonJellyfish, parameterName=NJF_Glow_Mask_01, sortPriority=399, texture=&T_BlacksRGB}
  n76 = TextureSample(
   UVs=Add(
    A=MaterialFunctionCall!(
     ForwardAxis
     Bounds=@n42
     Pre-Skinned Local Position
     isVehicle=@n3
     isMech=@n7
     Side Project Vehicle=@n27
     WeaponOffset=@n67
     WeaponFalloff
     VehicleOffset=@n67
     VehicleFalloff
     MechOffset=@n67
     MechFalloff
     Invert_Gradent
     UV_Spread_XY=@n68.RGB
     UV_Animated_XY=@n71
     Warp_UVs=@n72
     Rotate_Mech
     Rotate_Vehicle_Side
     Rotate_Vehicle=@n73
    ) {materialFunction=&/MF_Wrap_Gradent_Creation}.Uvs
    B=@n74
   )
   Tex=@n75
  )
  n77 = Saturate(None=Divide(A=@n9, B=ScalarParameter() {defaultValue=0.8, group=Wrap_NeonJellyfish, parameterName=NJF_Time_Duration, sortPriority=601}))
  n78 = MaterialFunctionCall(
   Texture=@n33
   Rotation_Axis=@n29
   Caculated_Bounds=Add(A=@n48, B=@n14)
   Pre_Skinned_Normals=@n37
   Blend_Contrast=@n31
   Fill_alternate_Axis_Opacity=@n34
   "Fill_alternate_Axis_Color_A=Roughness"=@n36
   Is_Vehicle=@n3
   Is_Mech=@n7
  ) {materialFunction=&MF_TopsSideProjection_Base}
  n79 = Multiply(
   A=ScalarParameter() {defaultValue=0.5, group=Wrap_NeonJellyfish, parameterName=NJF_Mask_03B_Warp_Amount, sortPriority=211}
   B=Add(A=ComponentMask(None=@n78.Combined) {r=true}) {constB=-0.25}
  )
  n80 = MaterialFunctionCall(
   Texture=@n32
   Rotation_Axis=@n29
   Caculated_Bounds=Add(A=@n63, B=Multiply(A=@n79, B=@n56.Raw_Data))
   Pre_Skinned_Normals=@n37
   Blend_Contrast=@n31
   Fill_alternate_Axis_Opacity=@n30
   "Fill_alternate_Axis_Color_A=Roughness"=@n23
   Is_Vehicle=@n3
   Is_Mech=@n7
  ) {materialFunction=&MF_TopsSideProjection_Base}
  n81 = ComponentMask(None=@n80.Combined) {r=true}
  n82 = Power(Base=Multiply(A=@n81) {constB=10}) {constExponent=5}
  n83 = VectorParameter() {group=Wrap_NeonJellyfish, parameterName=NJF_Glow_1_Spread, sortPriority=400, defaultValue=<rgba 1 1 0 1>}
  n84 = VectorParameter() {group=Wrap_NeonJellyfish, parameterName=NJF_Glow_Speed_1_Weapon, sortPriority=410, defaultValue=<rgba 0.5 0 0 1>}
  n85 = VectorParameter() {group=Wrap_NeonJellyfish, parameterName=NJF_Glow_Speed_1_Vehicle, sortPriority=410, defaultValue=<rgba 0 0 0 1>}
  n86 = LinearInterpolate(A=@n84.RGB, B=@n85.RGB, Alpha=@n3)
  n87 = Multiply(
   A=ScalarParameter() {defaultValue=0.5, group=Wrap_NeonJellyfish, parameterName=NJF_Glow_Warp_1_uvs, sortPriority=415}
   B=ComponentMask(None=@n80.RGB) {g=true}
  )
  n88 = TextureSample(
   UVs=Add(
    A=MaterialFunctionCall(
     Bounds=@n42
     isVehicle=@n3
     isMech=@n7
     Side Project Vehicle=@n27
     UV_Spread_XY=@n83.RGB
     UV_Animated_XY=@n86
     Warp_UVs=@n87
     Rotate_Vehicle=@n73
    ) {materialFunction=&MF_Wrap_Gradent_Creation}.Uvs
    B=@n74
   )
   Tex=@n75
  )
  n89 = Add(
   A=Multiply(A=LinearInterpolate(A=@n76.G, B=@n76.G, Alpha=@n77), B=Saturate(None=@n66))
   B=Multiply(A=LinearInterpolate(A=@n88.G, B=@n88.G, Alpha=@n77), B=Saturate(None=@n82))
  )
  n90 = ScalarParameter() {group=Wrap_NeonJellyfish, parameterName=NJF_Head_Roll_Offset, sliderMin=-1, sortPriority=-1}
  n91 = MaterialFunctionCall(Bounds=@n42, ForwardAxis=@n55, Offset=Add(A=Constant() {r=-0.92}, B=@n90), FallOff=Constant() {r=-0.1}) {materialFunction=&MF_Wrap_Gradent_Simple}
  n92 = Multiply(A=Multiply(A=@n65) {constB=0.5}, B=OneMinus(None=@n81))
  n93 = Constant() {r=-0.79}
  n94 = MaterialFunctionCall(
   Transition Phase=MaterialFunctionCall(Bounds=@n42, ForwardAxis=@n55, Offset=Add(A=@n90, B=@n93), FallOff=Constant() {r=0.2}) {materialFunction=&MF_Wrap_Gradent_Simple}.Raw_Data
   Height Texture=ComponentMask(None=@n78.Combined) {b=true}
   Contrast=Constant() {r=1}
  ) {materialFunction=&HeightLerp}
  n95 = Add(A=Multiply(A=Add(A=@n92, B=@n81), B=OneMinus(None=@n94.Alpha)), B=@n94.Alpha)
  n96 = CurveAtlasRowParameter(CurveTime=@n95) {
   atlas=/Game/Weapons/WeaponSkins/Wraps/Textures/CruveAtlas_Wraps_01
   curve=/Game/Weapons/WeaponSkins/Wraps/Textures/NeonJellyfish/CC_NeonJellyfish
   defaultValue=25
   group=Wrap_NeonJellyfish
   parameterName=NJF_ColorCurve
   sortPriority=399
  }
  n97 = MaterialFunctionCall!(
   Index=ScalarParameter() {defaultValue=25, group=Wrap_NeonJellyfish, parameterName="NJF_ColorCurve  Index", sortPriority=399}
   CCinput=@n96
   CCinput_Alpha=@n96
   UV=@n95
  ) {materialFunction=&/MF_Global_CC_Tex}
  n98 = Multiply(
   A=Multiply(
    A=ScalarParameter() {defaultValue=5, group=Wrap_NeonJellyfish, parameterName=NJF_Glow_Brightness, sortPriority=399}
    B=Multiply(A=@n89, B=@n91.Clamped_Data)
   )
   B=@n97.3V
  )
  n99 = Multiply(
   A=@n97.3V
   B=Multiply(A=@n97.Alpha, B=ScalarParameter() {defaultValue=1, group=Wrap_NeonJellyfish, parameterName=NJF_ColorCuve_Brightness, sortPriority=399})
  )
  n100 = ScalarParameter() {defaultValue=10, group=Wrap_NeonJellyfish, parameterName=NJF_Brightness_Rect, sortPriority=399}
  n101 = SetMaterialAttributes!(
   MaterialAttributes=@n20.MaterialAttributes
   BaseColor=@n21.RGB
   Roughness=ScalarParameter() {defaultValue=0.3, group=Wrap_NeonJellyfish, parameterName=NJF_Roghness, sortPriority=0}
   Specular=ScalarParameter() {defaultValue=0.5, group=Wrap_NeonJellyfish, parameterName=NJF_Specular, sortPriority=0}
   EmissiveColor=Multiply(A=Add(A=@n98, B=@n99), B=LinearInterpolate(A=@n100, Alpha=@n77) {constA=10})
  ) {attributeSetTypes [
    #1
    #2
    #4
    #5
   ]}
  n102 = Saturate(None=Add(
    A=FunctionInput() {bUsePreviewValueAsDefault=true, inputName="Customization Mask (_CM B)", inputType=FunctionInput_Scalar, sortPriority=24}
    B=ScalarParameter() {group=Wrap_NeonJellyfish, parameterName=NJF_Remove_Customization Mask, sortPriority=-10}
   ))
  FunctionOutput(None=StaticSwitchParameter(True=BlendMaterialAttributes(A=@n20.MaterialAttributes, B=@n101, Alpha=@n102) {vertexAttributeBlendType=UseA}, False=@n20.MaterialAttributes) {group=Customization_Wrap, parameterName=Wrap_NeonJellyfish})
  Multiply(A=@n94.Alpha, B=Constant() {r=1})
 ]}
