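# Weapon-wrap material function that draws a flickering neon Christmas sign on weapons, vehicles, mechs and characters, behind a static switch.
# Side-projected UVs sample two mask textures for six neon colors, pulsed by offset sine timers within a third mask's sections.
# A Blueprint-fed timestamp flashes all colors together, fading over a set duration.
# Output is time-of-day-scaled emissive or color-only, limited by normal-based and customization masks.
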MaterialFunction /Game/Weapons/WeaponSkins/Wraps/MaterialFunctions/MF_Wrap_NeonXmasSign
graph {nodes [
  ComponentMask() {g=true, r=true}
  Constant3Vector() {constant=<rgba 0.0178 0.0184 0.0273 1>}
  n1 = FunctionInput() {inputName=Material, inputType=FunctionInput_MaterialAttributes}
  n2 = GetMaterialAttributes(None=@n1) {attributeGetTypes [#1]}
  n3 = VectorParameter() {group=Wrap_NeonXmasSign, parameterName=NXS_BaseColor, sortPriority=0, defaultValue=<rgba 0.0178 0.0184 0.0273 1>}
  n4 = TextureObjectParameter() {
   group=Wrap__NeonXmasSign_Textures
   parameterName=NXS_Normal_01
   samplerType=SAMPLERTYPE_Normal
   sortPriority=13
   texture=&FlatNormalBlue
  }
  n5 = FunctionInput() {bUsePreviewValueAsDefault=true, inputName=isState, inputType=FunctionInput_Vector4, sortPriority=4, previewValue=<0 0 0 0>}
  n6 = MaterialFunctionCall(Color=@n5) {materialFunction=&SplitComponents}
  n7 = VectorParameter() {group=Wrap_NeonXmasSign_UVs, parameterName=NXS_Offset_Weapon, defaultValue=<rgba 0 0 0 1>}
  n8 = VectorParameter() {desc=must activate TS_Vehicle_Side_01 UVs, group=Wrap_NeonXmasSign_UVs, parameterName=NXS_Offset_Vehicle_Side, defaultValue=<rgba 0 0 0 1>}
  n9 = MaterialFunctionCall(Float4=@n5) {materialFunction=&BreakOutFloat4Components}
  n10 = LinearInterpolate(A=VectorParameter() {group=Wrap_NeonXmasSign_UVs, parameterName=NXS_Offset_Vehicle, defaultValue=<rgba 0 0 0 1>}.RGB, B=@n8.RGB, Alpha=@n9.G)
  n11 = VectorParameter() {group=Wrap_NeonXmasSign_UVs, parameterName=NXS_Offset_Mech -----, defaultValue=<rgba 0 0 0 1>}
  n12 = VectorParameter() {group=Wrap_NeonXmasSign_UVs, parameterName=NXS_Offset_Character, defaultValue=<rgba 0 0 0 1>}
  n13 = ScalarParameter() {defaultValue=1, group=Wrap_NeonXmasSign_UVs, parameterName=NXS_TileRate_Weapon}
  n14 = LinearInterpolate(
   A=ScalarParameter() {defaultValue=1, group=Wrap_NeonXmasSign_UVs, parameterName=NXS_TileRate_Vehicle}
   B=ScalarParameter() {defaultValue=1, group=Wrap_NeonXmasSign_UVs, parameterName=NXS_TileRate_VehicleSide}
   Alpha=@n9.G
  )
  n15 = ScalarParameter() {defaultValue=1, group=Wrap_NeonXmasSign_UVs, parameterName=NXS_TileRate_Mech -----}
  n16 = ScalarParameter() {defaultValue=1, group=Wrap_NeonXmasSign_UVs, parameterName=NXS_TileRate_Character}
  n17 = FunctionInput(Preview=VertexInterpolator(VS=PreSkinnedNormal()).PS) {bUsePreviewValueAsDefault=true, inputName=Pre-Skinned Local Normal, sortPriority=10}
  n18 = VertexInterpolator(VS=PreSkinnedPosition())
  n19 = PreSkinnedLocalBounds()
  n20 = FunctionInput(Preview=Divide(A=Subtract(A=@n18.PS, B=@n19.Min), B=Subtract(A=@n19.Max, B=@n19.Max))) {bUsePreviewValueAsDefault=true, inputName=Calcated_Bounds, sortPriority=11}
  n21 = MaterialFunctionCall!(
   For_Safe_Zones
   Normal_Map=@n4
   Caculated_Bounds=@n20
   Pre_Skinned_Normals=@n17
   Is Vehicle=@n6.R
   Is VehicleSide=@n6.R
   Is Mech=@n6.R
   Vehicle_Rotation=ScalarParameter() {group=Wrap_NeonXmasSign_UVs, parameterName=NXS_Rotation_Vehicle}
   UV_Rotation Bypass
   Mech_Rotation=ScalarParameter() {defaultValue=0.25, group=Wrap_NeonXmasSign_UVs, parameterName=NXS_Rotation_Mech -----}
   Uvs_Offset=LinearInterpolate(A=LinearInterpolate(A=@n7.RGB, B=@n10, Alpha=@n9.R), B=LinearInterpolate(A=@n11.RGB, B=@n12.RGB, Alpha=@n9.A), Alpha=@n9.B)
   Offset-after rotation
   Uvs_TileRate=LinearInterpolate(A=LinearInterpolate(A=@n13, B=@n14, Alpha=@n9.R), B=LinearInterpolate(A=@n15, B=@n16, Alpha=@n9.A), Alpha=@n9.B)
  ) {materialFunction=&/MF_Projection_Side_Normal}
  n22 = TextureObjectParameter() {
   group=Wrap__NeonXmasSign_Textures
   parameterName=NXS_BaseColor - No Alpha
   sortPriority=10
   texture=&T_BlacksRGB
  }
  n23 = ScalarParameter() {group=Wrap_NeonXmasSign, parameterName=NXS_Curve_Animated, sortPriority=-50}
  n24 = Time()
  n25 = ScalarParameter() {defaultValue=1, group=Wrap_NeonXmasSign, parameterName=NXS_Curve_TileRate, sortPriority=-50}
  n26 = Constant3Vector() {constant=<rgba 1 0 0 0>}
  n27 = LinearInterpolate(
   A=ScalarParameter() {group=Wrap_NeonXmasSign, parameterName=NXS_Curve_Offset_Weapon, sortPriority=-50}
   B=ScalarParameter() {group=Wrap_NeonXmasSign, parameterName=NXS_Curve_Offset_Vehicle, sortPriority=-50}
   Alpha=@n9.R
  )
  n28 = ScalarParameter() {group=Wrap_NeonXmasSign, parameterName=NXS_Curve_Offset_Mech, sortPriority=-50}
  n29 = ScalarParameter() {defaultValue=1, desc="negative value inverts ", group=Wrap_NeonXmasSign, parameterName=NXS_Curve_Falloff_Weapon, sortPriority=-50}
  n30 = ScalarParameter() {defaultValue=1, desc="negative value inverts ", group=Wrap_NeonXmasSign, parameterName=NXS_Curve_Falloff_Vehicle, sortPriority=-50}
  n31 = LinearInterpolate(A=@n29, B=@n30, Alpha=@n9.R)
  n32 = ScalarParameter() {defaultValue=1, desc="negative value inverts ", group=Wrap_NeonXmasSign, parameterName=NXS_Curve_Falloff_Mech, sortPriority=-50}
  n33 = Frac(None=MaterialFunctionCall!(
    Bounds=@n20
    ForwardAxis=FunctionInput(Preview=@n26) {bUsePreviewValueAsDefault=true, inputName=ForwardAxis}
    Offset=LinearInterpolate(A=@n27, B=@n28, Alpha=@n9.B)
    FallOff=LinearInterpolate(A=@n31, B=@n32, Alpha=@n9.B)
   ) {materialFunction=&/MF_Wrap_Gradent_Simple}.Raw_Data)
  n34 = Frac(None=Add(A=Multiply(A=@n23, B=@n24), B=Multiply(A=@n25, B=@n33)))
  n35 = CurveAtlasRowParameter(CurveTime=@n34) {
   atlas=/Game/Weapons/WeaponSkins/Wraps/Textures/CruveAtlas_Wraps_01
   curve=/Game/Weapons/WeaponSkins/Wraps/Textures/C_GradSimple
   group=Wrap_NeonXmasSign
   parameterName=NXS_Curve
   sortPriority=-50
  }
  n36 = MaterialFunctionCall!(
   Index=ScalarParameter() {group=Wrap_NeonXmasSign, parameterName="NXS_Curve  Index", sortPriority=-50}
   CCinput=@n35
   CCinput_Alpha=@n35
   UV=@n34
  ) {materialFunction=&/MF_Global_CC_Tex}
  n37 = Multiply(A=@n36.Alpha, B=ScalarParameter() {defaultValue=1, group=Wrap_NeonXmasSign, parameterName=NXS_CurveBrightness, sortPriority=-50})
  n38 = LinearInterpolate(
   A=@n3.RGB
   B=StaticSwitchParameter(True=TextureSample(UVs=@n21.UVs, Tex=@n22).RGB, False=Multiply(A=@n36.3V, B=@n37)) {defaultValue=true, group=Wrap_NeonXmasSign, parameterName="NXS_BaseColor_UseTexture=True - Curve=False", sortPriority=-102}
   Alpha=ScalarParameter() {group=Wrap_NeonXmasSign, parameterName=NXS_BaseColor_useTextureorCurve - Set to 1, sortPriority=-103}
  )
  n39 = VectorParameter() {group=Wrap_NeonXmasSign, parameterName=NXS_Mask01_Color_R, sortPriority=1, defaultValue=<rgba 0.9911 0.0097 0 1>}
  n40 = TextureObjectParameter() {
   group=Wrap__NeonXmasSign_Textures
   parameterName="NXS_Mask_01 - RGB=Color Masks A=Roughness"
   sortPriority=10
   texture=&T_BlacksRGB
  }
  n41 = TextureSample(UVs=@n21.UVs, Tex=@n40)
  n42 = ComponentMask(None=@n41.RGBA) {r=true}
  n43 = LinearInterpolate(A=@n38, B=@n39.RGB, Alpha=@n42)
  n44 = ScalarParameter() {defaultValue=0.025, group=Wrap_NeonXmasSign, parameterName=NXS_BaseColorBlend - Blends base color with neon colors, sortPriority=2}
  n45 = Multiply(A=ScalarParameter() {defaultValue=0.5, group=Wrap_NeonXmasSign, parameterName=NXS_TimingVariation_MaskMap01, sortPriority=49}, B=Time())
  n46 = Saturate(None=Sine(None=@n45) {period=2.1})
  n47 = Power(Base=@n46) {constExponent=1.5}
  n48 = VectorParameter() {group=Wrap_NeonXmasSign, parameterName=NXS_Mask01_Color_G, sortPriority=1, defaultValue=<rgba 0 0.7305 0.0482 1>}
  n49 = ComponentMask(None=@n41.RGBA) {g=true}
  n50 = LinearInterpolate(A=@n38, B=@n48.RGB, Alpha=@n49)
  n51 = Sine(None=@n45) {period=1.6}
  n52 = Sine(None=@n45) {period=1.42}
  n53 = Saturate(None=Desaturation(None=AppendVector(A=@n51, B=@n52)))
  n54 = Power(Base=@n53) {constExponent=0.5}
  n55 = LinearInterpolate(
   A=LinearInterpolate(A=LinearInterpolate(A=@n43, B=@n38, Alpha=@n44), B=@n39.RGB, Alpha=Multiply(A=@n42, B=@n47))
   B=LinearInterpolate(A=LinearInterpolate(A=@n50, B=@n38, Alpha=@n44), B=@n48.RGB, Alpha=Multiply(A=@n49, B=@n54))
   Alpha=@n49
  )
  n56 = VectorParameter() {group=Wrap_NeonXmasSign, parameterName=NXS_Mask01_Color_B, sortPriority=1, defaultValue=<rgba 0 0.7529 0.7682 1>}
  n57 = ComponentMask(None=@n41.RGBA) {b=true}
  n58 = Sine(None=@n45) {period=1.51}
  n59 = Saturate(None=Desaturation(None=AppendVector(A=@n51, B=@n58)))
  n60 = LinearInterpolate(
   A=LinearInterpolate(A=LinearInterpolate(A=@n38, B=@n56.RGB, Alpha=@n57), B=@n38, Alpha=@n44)
   B=@n56.RGB
   Alpha=Multiply(A=@n57, B=Power(Base=@n59) {constExponent=0.5})
  )
  n61 = VectorParameter() {group=Wrap_NeonXmasSign, parameterName=NXS_Mask02_Color_R, sortPriority=1, defaultValue=<rgba 0.8879 0.8796 0.956 1>}
  n62 = TextureObjectParameter() {
   group=Wrap__NeonXmasSign_Textures
   parameterName="NXS_Mask_02 - RGB Color Masks A=Metal Mask"
   sortPriority=11
   texture=&T_BlacksRGB
  }
  n63 = TextureSample(UVs=@n21.UVs, Tex=@n62)
  n64 = ComponentMask(None=@n63.RGBA) {r=true}
  n65 = LinearInterpolate(A=LinearInterpolate(A=@n38, B=@n61.RGB, Alpha=@n64), B=@n38, Alpha=@n44)
  n66 = Multiply(A=ScalarParameter() {defaultValue=0.8, group=Wrap_NeonXmasSign, parameterName=NXS_TimingVariation_MaskMap02, sortPriority=49}, B=Time())
  n67 = Saturate(None=Sine(None=@n66) {period=2})
  n68 = Multiply(A=@n64, B=Power(Base=@n67) {constExponent=1})
  n69 = VectorParameter() {group=Wrap_NeonXmasSign, parameterName=NXS_Mask02_Color_G, sortPriority=1, defaultValue=<rgba 1 0.9301 0.0369 1>}
  n70 = ComponentMask(None=@n63.RGBA) {g=true}
  n71 = LinearInterpolate(A=@n38, B=@n69.RGB, Alpha=@n70)
  n72 = Sine(None=@n66) {period=1.4}
  n73 = Sine(None=@n66) {period=1.33}
  n74 = Saturate(None=Desaturation(None=AppendVector(A=@n72, B=@n73)))
  n75 = Power(Base=@n74) {constExponent=0.5}
  n76 = LinearInterpolate(
   A=LinearInterpolate(A=LinearInterpolate(A=@n55, B=@n60, Alpha=@n57), B=LinearInterpolate(A=@n65, B=@n61.RGB, Alpha=@n68), Alpha=@n64)
   B=LinearInterpolate(A=LinearInterpolate(A=@n71, B=@n38, Alpha=@n44), B=@n69.RGB, Alpha=Multiply(A=@n70, B=@n75))
   Alpha=@n70
  )
  n77 = VectorParameter() {group=Wrap_NeonXmasSign, parameterName=NXS_Mask02_Color_B, sortPriority=1, defaultValue=<rgba 0.6172 0.0152 0.4852 1>}
  n78 = ComponentMask(None=@n63.RGBA) {b=true}
  n79 = Sine(None=@n66) {period=1.5}
  n80 = Saturate(None=Desaturation(None=AppendVector(A=@n72, B=@n79)))
  n81 = LinearInterpolate(
   A=LinearInterpolate(A=LinearInterpolate(A=@n38, B=@n77.RGB, Alpha=@n78), B=@n38, Alpha=@n44)
   B=@n77.RGB
   Alpha=Multiply(A=@n78, B=Power(Base=@n80) {constExponent=0.5})
  )
  n82 = TextureObjectParameter() {
   group=Wrap__NeonXmasSign_Textures
   parameterName="NXS_Mask_03 - RGB=Mask sections A=Specular"
   sortPriority=12
   texture=&T_BlacksRGB
  }
  n83 = TextureSample(UVs=@n21.UVs, Tex=@n82)
  n84 = ComponentMask(None=@n83.RGBA) {b=true}
  n85 = LinearInterpolate(A=LinearInterpolate(A=LinearInterpolate(A=@n38, B=@n39.RGB, Alpha=@n42), B=@n48.RGB, Alpha=@n49), B=@n56.RGB, Alpha=@n57)
  n86 = LinearInterpolate(A=LinearInterpolate(A=LinearInterpolate(A=@n85, B=@n61.RGB, Alpha=@n64), B=@n69.RGB, Alpha=@n70), B=@n77.RGB, Alpha=@n78)
  n87 = LinearInterpolate(A=@n38, B=@n86, Alpha=@n44) {desc=all color blend}
  n88 = Multiply(A=ScalarParameter() {defaultValue=0.25, group=Wrap_NeonXmasSign, parameterName=NXS_TimingVariation_MaskMap03, sortPriority=49}, B=Time())
  n89 = Sine(None=@n88) {period=1.4}
  n90 = Sine(None=@n88) {period=1.33}
  n91 = Saturate(None=Desaturation(None=AppendVector(A=@n89, B=@n90)))
  n92 = ComponentMask(None=@n83.RGBA) {g=true}
  n93 = LinearInterpolate(
   A=LinearInterpolate(A=@n87, B=LinearInterpolate(A=@n76, B=@n81, Alpha=@n78), Alpha=@n84)
   B=LinearInterpolate(A=@n87, B=@n86, Alpha=Power(Base=@n91) {constExponent=0.25})
   Alpha=@n92
  )
  n94 = Sine(None=@n88) {period=1.5}
  n95 = Saturate(None=Desaturation(None=AppendVector(A=@n89, B=@n94)))
  n96 = Power(Base=@n95) {constExponent=0.25}
  n97 = ComponentMask(None=@n83.RGBA) {r=true}
  n98 = Time()
  n99 = ScalarParameter() {defaultValue=-9999999, group=Wrap_NeonXmasSign_Reactive, parameterName=TimeFrom-BP----------------, sortPriority=-600}
  n100 = Saturate(None=Divide(
    A=Subtract(A=@n98, B=@n99)
    B=ScalarParameter() {defaultValue=1, group=Wrap_NeonXmasSign_Reactive, parameterName=NXS_AnimationDuration, sortPriority=33}
   ))
  n101 = ScalarParameter() {defaultValue=0.5, group=Wrap_NeonXmasSign_Reactive, parameterName=NXS_Curve_Amount, sortPriority=34}
  n102 = ScalarParameter() {group=Wrap_NeonXmasSign_Reactive, parameterName=NXS_TestReactive, sortPriority=300}
  n103 = LinearInterpolate(A=OneMinus(None=Power(Base=@n100, Exp=@n101)) {desc=inverts curve from 0-1 to 1-0}, B=@n102, Alpha=@n102)
  n104 = LinearInterpolate(A=LinearInterpolate(A=@n93, B=LinearInterpolate(A=@n93, B=@n86, Alpha=@n96), Alpha=@n97), B=@n86, Alpha=@n103)
  n105 = Add(A=@n42, B=@n49)
  n106 = Add(A=@n64, B=@n70)
  n107 = Saturate(None=Add(A=Add(A=@n105, B=@n57), B=Add(A=@n78, B=@n106)))
  n108 = MaterialFunctionCall(Color=Abs(None=@n17)) {materialFunction=&SplitComponents}
  n109 = MaterialFunctionCall(Color=@n5) {materialFunction=&SplitComponents}
  n110 = LinearInterpolate(A=Add(A=@n108.R, B=@n108.R), B=Add(A=@n108.B, B=@n108.B), Alpha=@n109.R)
  n111 = Add(A=@n108.R, B=@n108.R)
  n112 = ScalarParameter() {defaultValue=0.1, group=Wrap_NeonXmasSign, parameterName=NXS_Mask_Contrast, sortPriority=51}
  n113 = OneMinus(None=MaterialFunctionCall(In=LinearInterpolate(A=@n110, B=@n111, Alpha=@n109.B), Contrast=@n112) {materialFunction=&CheapContrast}.Result)
  n114 = MaterialFunctionCall(Color=Abs(None=@n17)) {materialFunction=&SplitComponents}
  n115 = LinearInterpolate(A=@n114.R, B=LinearInterpolate(A=@n114.B, B=@n114.B, Alpha=@n109.G), Alpha=@n109.R)
  n116 = OneMinus(None=MaterialFunctionCall(In=LinearInterpolate(A=@n115, B=@n114.G, Alpha=@n109.B), Contrast=@n112) {materialFunction=&CheapContrast}.Result)
  n117 = ScalarParameter() {group=Wrap_NeonXmasSign, parameterName=NXS_Mask_Z+AltAxisActive -----, sortPriority=50}
  n118 = LinearInterpolate(
   A=LinearInterpolate(
    A=Constant()
    B=Saturate(None=@n113)
    Alpha=ScalarParameter() {group=Wrap_NeonXmasSign, parameterName=NXS_Mask_AltAxisActive -----, sortPriority=50}
   )
   B=LinearInterpolate(A=Constant(), B=Saturate(None=@n116), Alpha=@n117)
   Alpha=@n117
  )
  n119 = Add(A=@n2.Normal, B=Clamp(None=@n21.Raw_Normal))
  n120 = ScalarParameter() {defaultValue=1, group=Wrap_NeonXmasSign, parameterName=NXS_NormalIntensity, sortPriority=5}
  n121 = StaticSwitchParameter(True=@n2.Normal, False=LinearInterpolate(A=LinearInterpolate(A=@n2.Normal, B=@n119, Alpha=@n120), B=@n2.Normal, Alpha=@n118)) {group=Wrap_NeonXmasSign, parameterName=NXS_No Normal, sortPriority=-105}
  n122 = LinearInterpolate(
   A=LinearInterpolate(
    A=ScalarParameter() {defaultValue=0.5, group=Wrap_NeonXmasSign, parameterName=NXS_Specular_Min}
    B=ScalarParameter() {defaultValue=0.5, group=Wrap_NeonXmasSign, parameterName=NXS_Specular_Max}
    Alpha=ComponentMask(None=@n83.RGBA) {a=true}
   )
   B=ScalarParameter() {defaultValue=0.5, group=Wrap_NeonXmasSign, parameterName=NXS_Mask_Specular, sortPriority=52}
   Alpha=@n118
  )
  n123 = LinearInterpolate(
   A=LinearInterpolate(
    A=ScalarParameter() {group=Wrap_NeonXmasSign, parameterName=NXS_Roughness_Min}
    B=ScalarParameter() {defaultValue=1, group=Wrap_NeonXmasSign, parameterName=NXS_Roughness_Max}
    Alpha=ComponentMask(None=@n41.RGBA) {a=true}
   )
   B=ScalarParameter() {defaultValue=0.5, group=Wrap_NeonXmasSign, parameterName=NXS_Mask_Roughness, sortPriority=52}
   Alpha=@n118
  )
  n124 = LinearInterpolate(
   A=LinearInterpolate(
    A=ScalarParameter() {group=Wrap_NeonXmasSign, parameterName=NXS_Metallic_Min}
    B=ScalarParameter() {group=Wrap_NeonXmasSign, parameterName=NXS_Metallic_Max}
    Alpha=ComponentMask(None=@n63.RGBA) {a=true}
   )
   B=ScalarParameter() {group=Wrap_NeonXmasSign, parameterName=NXS_Mask_Metallic, sortPriority=52}
   Alpha=@n118
  )
  n125 = SetMaterialAttributes!(
   MaterialAttributes=@n2.MaterialAttributes
   BaseColor=LinearInterpolate(A=LinearInterpolate(A=@n38, B=@n104, Alpha=@n107), B=@n38, Alpha=@n118)
   Roughness=@n123
   Normal=@n121
   Specular=@n122
   Metallic=@n124
  ) {attributeSetTypes [
    #2
    #3
    #1
    #4
    #5
   ]}
  n126 = Multiply(A=@n39.RGB, B=@n42)
  n127 = ScalarParameter() {defaultValue=2, group=Wrap_NeonXmasSign, parameterName=NXS_Glow_Intensity -----, sortPriority=33}
  n128 = ScalarParameter() {defaultValue=1, group=Wrap_NeonXmasSign, parameterName=NXS_Glow_Max, sortPriority=49}
  n129 = ScalarParameter() {group=Wrap_NeonXmasSign, parameterName=NXS_Glow_Min, sortPriority=49}
  n130 = Multiply(A=@n48.RGB, B=@n49)
  n131 = Add(
   A=LinearInterpolate(B=Multiply(A=@n126, B=@n127), Alpha=Clamp(None=@n46, Min=@n129, Max=@n128))
   B=LinearInterpolate(B=Multiply(A=@n130, B=@n127), Alpha=Clamp(None=@n53, Min=@n129, Max=@n128))
  )
  n132 = Multiply(A=@n56.RGB, B=@n57)
  n133 = LinearInterpolate(B=Multiply(A=@n132, B=@n127), Alpha=Clamp(None=@n59, Min=@n129, Max=@n128))
  n134 = Multiply(A=@n61.RGB, B=@n64)
  n135 = Multiply(A=@n134, B=@n127)
  n136 = Clamp(None=@n67, Min=@n129, Max=@n128)
  n137 = Multiply(A=@n69.RGB, B=@n70)
  n138 = Add(
   A=Add(A=Add(A=@n131, B=@n133), B=LinearInterpolate(B=@n135, Alpha=@n136))
   B=LinearInterpolate(B=Multiply(A=@n137, B=@n127), Alpha=Clamp(None=@n74, Min=@n129, Max=@n128))
  )
  n139 = Multiply(A=@n77.RGB, B=@n78)
  n140 = LinearInterpolate(B=Multiply(A=@n139, B=@n127), Alpha=Clamp(None=@n80, Min=@n129, Max=@n128)) {desc=red}
  n141 = Add(A=Add(A=Add(A=@n126, B=@n130), B=@n132), B=Add(A=Add(A=@n134, B=@n137), B=@n139))
  n142 = Multiply(A=@n141, B=@n127)
  n143 = LinearInterpolate(
   A=LinearInterpolate(B=Add(A=@n138, B=@n140), Alpha=@n84)
   B=LinearInterpolate(B=@n142, Alpha=Clamp(None=@n91, Min=@n129, Max=@n128))
   Alpha=@n92
  )
  n144 = LinearInterpolate(B=@n142, Alpha=Clamp(None=@n95, Min=@n129, Max=@n128))
  n145 = ScalarParameter() {defaultValue=5, group=Wrap_NeonXmasSign_Reactive, parameterName=NXS_EmissiveReactiveIntensity, sortPriority=0}
  n146 = ScalarParameter() {defaultValue=0.5, group=Wrap_NeonXmasSign_Emissive, parameterName=NXS_TOD_EmissiveMorningEvening}
  n147 = Multiply(
   A=LinearInterpolate(A=LinearInterpolate(A=@n143, B=@n144, Alpha=@n97), B=Multiply(A=@n141, B=@n145), Alpha=@n103)
   B=MaterialFunctionCall(
    Value_Morning=@n146
    Value_Day=ScalarParameter() {defaultValue=1, group=Wrap_NeonXmasSign_Emissive, parameterName=NXS_TOD_EmissiveDay}
    Value_Evening=@n146
    Value_Night=ScalarParameter() {defaultValue=0.25, group=Wrap_NeonXmasSign_Emissive, parameterName=NXS_TOD_EmissiveNight}
   ) {materialFunction=&Scalar_Time_of_Day}.Result
  )
  n148 = Multiply(A=ScalarParameter() {defaultValue=1, group=Wrap_NeonXmasSign, parameterName=NXS_DarkenBaseWhenEmissive-------, sortPriority=-1}, B=@n147)
  n149 = OneMinus(None=Saturate(None=Desaturation(None=@n148)))
  n150 = LinearInterpolate(A=@n38, B=Multiply(A=@n42, B=LinearInterpolate(A=@n38, B=@n39.RGB, Alpha=@n44)), Alpha=@n42)
  n151 = Multiply(A=@n49, B=LinearInterpolate(A=@n38, B=@n48.RGB, Alpha=@n44))
  n152 = LinearInterpolate(A=@n38, B=@n56.RGB, Alpha=@n44)
  n153 = LinearInterpolate(
   A=LinearInterpolate(A=LinearInterpolate(A=@n150, B=@n151, Alpha=@n49), B=Multiply(A=@n57, B=@n152), Alpha=@n57)
   B=Multiply(A=@n64, B=LinearInterpolate(A=@n38, B=@n61.RGB, Alpha=@n44))
   Alpha=@n64
  )
  n154 = Multiply(A=@n70, B=LinearInterpolate(A=@n38, B=@n69.RGB, Alpha=@n44))
  n155 = LinearInterpolate(A=@n38, B=@n77.RGB, Alpha=@n44)
  n156 = LinearInterpolate(A=LinearInterpolate(A=LinearInterpolate(A=@n153, B=@n154, Alpha=@n70), B=Multiply(A=@n78, B=@n155), Alpha=@n78), B=@n38, Alpha=@n118)
  n157 = OneMinus(None=@n118)
  n158 = SetMaterialAttributes!(
   MaterialAttributes=@n2.MaterialAttributes
   BaseColor=LinearInterpolate(A=@n156, B=Multiply(A=@n156, B=@n149), Alpha=@n157) {constAlpha=1}
   Roughness=@n123
   Normal=@n121
   Specular=@n122
   Metallic=@n124
   EmissiveColor=Multiply(A=@n157, B=@n147)
  ) {attributeSetTypes [
    #2
    #3
    #1
    #4
    #5
    #6
   ]}
  n159 = FunctionInput() {bUsePreviewValueAsDefault=true, inputName="Customization Mask (_CM B Top left)", inputType=FunctionInput_Scalar, sortPriority=24}
  n160 = FunctionInput() {bUsePreviewValueAsDefault=true, inputName="Customization Mask 2 (_CM B Top Right)", inputType=FunctionInput_Scalar, sortPriority=25}
  n161 = NamedRerouteUsage()
  n162 = NamedRerouteUsage()
  n163 = LinearInterpolate(A=@n160, Alpha=Saturate(None=Subtract(A=@n161, B=@n162))) {constB=0}
  n164 = StaticSwitchParameter(
   True=BlendMaterialAttributes(
    A=@n2.MaterialAttributes
    B=StaticSwitchParameter(True=@n125, False=@n158) {group=Wrap_NeonXmasSign, parameterName=NXS_UseBaseColorBlendOnly - Instead of Emissive, sortPriority=-110}
    Alpha=Add(A=@n159, B=@n163)
   )
   False=@n1
  ) {
   desc=Side Projection Base - Mask Driven with separate channels driving a Color - Channels glow offset timings - Reactive -Brighten all at once on Fire
   group=Customization_Wrap
   parameterName=Wrap_NeonXmasSign
  }
  FunctionOutput(None=@n164)
  MaterialFunctionCall!("Customization Mask (_CM B Top left)"=@n159, "Customization Mask 2 (_CM B Top Right)"=@n160) {materialFunction=&/MF_Customization_AllMasking}
  NamedRerouteDeclaration(None=@n9.A) {name=IsCharacter, nodeColor=<rgba 0.8958 1 0.0343 1>}
  NamedRerouteDeclaration(None=FunctionInput() {bUsePreviewValueAsDefault=true, inputName=isCharAccent, inputType=FunctionInput_Scalar, sortPriority=78}) {name=isCharAccent, nodeColor=<rgba 0 0.9253 1 1>}
  Reroute(None=@n38)
 ]}
