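# Material function for the switchable PinkToons wrap skin on weapons, vehicles and mechs, with a reactive glow on weapon fire.
# Side-projected base texture and three panning alpha-layered color textures drive color and masked surface properties, with optional normal map and axis masks.
# Time since a Blueprint-set timestamp drives decaying, time-of-day-scaled emissive that darkens base color.
# Customization masks blend it into incoming attributes, then an optional Fresnel-and-curvature outline is applied.
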MaterialFunction /Game/Weapons/WeaponSkins/Wraps/MaterialFunctions/MF_Wrap_PinkToons
graph {nodes [
  n1 = FunctionInput() {inputName=Material, inputType=FunctionInput_MaterialAttributes}
  n2 = GetMaterialAttributes(None=@n1) {attributeGetTypes [#1]}
  n3 = TextureObjectParameter() {
   group=Wrap__PinkToons_Textures
   parameterName=PT_Normal_01
   samplerType=SAMPLERTYPE_Normal
   sortPriority=13
   texture=&FlatNormalBlue
  }
  n4 = Constant()
  n5 = FunctionInput(Preview=@n4) {bUsePreviewValueAsDefault=true, inputName=IsState, inputType=FunctionInput_Vector4}
  n6 = MaterialFunctionCall(Color=@n5) {materialFunction=&SplitComponents}
  n7 = VectorParameter() {group=Wrap_PinkToons_UVs, parameterName=PT_Offset_Vehicle, defaultValue=<rgba 0 0 0 1>}
  n8 = VectorParameter() {desc=must activate TS_Vehicle_Side_01 UVs, group=Wrap_PinkToons_UVs, parameterName=PT_Offset_Vehicle_Side, defaultValue=<rgba 0 0 0 1>}
  n9 = MaterialFunctionCall(Color=@n5) {materialFunction=&SplitComponents}
  n10 = LinearInterpolate(
   A=LinearInterpolate(
    A=VectorParameter() {group=Wrap_PinkToons_UVs, parameterName=PT_Offset_Weapon, defaultValue=<rgba 0 0 0 1>}.RGB
    B=LinearInterpolate(A=@n7.RGB, B=@n8.RGB, Alpha=@n9.G)
    Alpha=@n9.R
   )
   B=VectorParameter() {group=Wrap_PinkToons_UVs, parameterName=PT_Offset_Mech -----, defaultValue=<rgba 0 0 0 1>}.RGB
   Alpha=@n9.B
  )
  n11 = ScalarParameter() {defaultValue=1, group=Wrap_PinkToons_UVs, parameterName=PT_TileRate_Vehicle}
  n12 = ScalarParameter() {defaultValue=1, group=Wrap_PinkToons_UVs, parameterName=PT_TileRate_VehicleSide}
  n13 = LinearInterpolate(
   A=LinearInterpolate(
    A=ScalarParameter() {defaultValue=1, group=Wrap_PinkToons_UVs, parameterName=PT_TileRate_Weapon}
    B=LinearInterpolate(A=@n11, B=@n12, Alpha=@n9.G)
    Alpha=@n9.R
   )
   B=ScalarParameter() {defaultValue=1, group=Wrap_PinkToons_UVs, parameterName=PT_TileRate_Mech -----}
   Alpha=@n9.B
  )
  n14 = FunctionInput(Preview=VertexInterpolator(VS=PreSkinnedNormal()).PS) {bUsePreviewValueAsDefault=true, inputName=Pre-Skinned Local Normal, sortPriority=10}
  n15 = VertexInterpolator(VS=PreSkinnedPosition())
  n16 = PreSkinnedLocalBounds()
  n17 = FunctionInput(Preview=Divide(A=Subtract(A=@n15.PS, B=@n16.Min), B=Subtract(A=@n16.Max, B=@n16.Max))) {bUsePreviewValueAsDefault=true, inputName=Calcated_Bounds, sortPriority=11}
  n18 = MaterialFunctionCall!(
   For_Safe_Zones
   Normal_Map=@n3
   Caculated_Bounds=@n17
   Pre_Skinned_Normals=@n14
   Is Vehicle=@n6.R
   Is VehicleSide=@n6.R
   Is Mech=@n6.R
   Vehicle_Rotation=ScalarParameter() {group=Wrap_PinkToons_UVs, parameterName=PT_Rotation_Vehicle}
   UV_Rotation Bypass
   Mech_Rotation=ScalarParameter() {defaultValue=0.25, group=Wrap_PinkToons_UVs, parameterName=PT_Rotation_Mech -----}
   Uvs_Offset=@n10
   Offset-after rotation
   Uvs_TileRate=@n13
  ) {materialFunction=&/MF_Projection_Side_Normal}
  n19 = LinearInterpolate(
   A=@n2.Normal
   B=Add(A=@n2.Normal, B=@n18.Raw_Normal)
   Alpha=ScalarParameter() {defaultValue=1, group=Wrap_PinkToons, parameterName=PT_NormalIntensity, sortPriority=5}
  )
  n20 = MaterialFunctionCall(Color=Abs(None=@n14)) {materialFunction=&SplitComponents}
  n21 = MaterialFunctionCall(Color=@n5) {materialFunction=&SplitComponents}
  n22 = LinearInterpolate(A=Add(A=@n20.R, B=@n20.R), B=Add(A=@n20.B, B=@n20.B), Alpha=@n21.R)
  n23 = Add(A=@n20.R, B=@n20.R)
  n24 = ScalarParameter() {defaultValue=0.1, group=Wrap_PinkToons, parameterName=PT_Mask_Contrast, sortPriority=51}
  n25 = OneMinus(None=MaterialFunctionCall(In=LinearInterpolate(A=@n22, B=@n23, Alpha=@n21.B), Contrast=@n24) {materialFunction=&CheapContrast}.Result)
  n26 = MaterialFunctionCall(Color=Abs(None=@n14)) {materialFunction=&SplitComponents}
  n27 = LinearInterpolate(A=@n26.R, B=LinearInterpolate(A=@n26.B, B=@n26.B, Alpha=@n21.G), Alpha=@n21.R)
  n28 = OneMinus(None=MaterialFunctionCall(In=LinearInterpolate(A=@n27, B=@n26.G, Alpha=@n21.B), Contrast=@n24) {materialFunction=&CheapContrast}.Result)
  n29 = ScalarParameter() {group=Wrap_PinkToons, parameterName=PT_Mask_Z+AltAxisActive -----, sortPriority=50}
  n30 = LinearInterpolate(
   A=LinearInterpolate(
    A=Constant()
    B=Saturate(None=@n25)
    Alpha=ScalarParameter() {group=Wrap_PinkToons, parameterName=PT_Mask_AltAxisActive -----, sortPriority=50}
   )
   B=LinearInterpolate(A=Constant(), B=Saturate(None=@n28), Alpha=@n29)
   Alpha=@n29
  )
  n31 = LinearInterpolate(
   A=LinearInterpolate(A=@n19, B=@n2.Normal, Alpha=@n30)
   B=@n19
   Alpha=ScalarParameter() {group=Wrap_PinkToons, parameterName=PT_Mask_NormalUseBase, sortPriority=55}
  )
  n32 = MaterialFunctionCall(Color=@n5) {materialFunction=&SplitComponents}
  n33 = VectorParameter() {group=Wrap_PinkToons_UVs, parameterName=PT_ColorTexture_01_Panning_Weapon, defaultValue=<rgba 0 0 0 1>}
  n34 = VectorParameter() {group=Wrap_PinkToons_UVs, parameterName=PT_ColorTexture_01_Panning_Vehicle, defaultValue=<rgba 0 0 0 1>}
  n35 = VectorParameter() {group=Wrap_PinkToons_UVs, parameterName=PT_ColorTexture_01_Panning_VehicleSide, defaultValue=<rgba 0 0 0 1>}
  n36 = LinearInterpolate(A=@n34.RGB, B=@n35.RGB)
  n37 = VectorParameter() {group=Wrap_PinkToons_UVs, parameterName=PT_ColorTexture_01_Panning_Mech -----, defaultValue=<rgba 0 0 0 1>}
  n38 = Multiply(A=Time(), B=LinearInterpolate(A=LinearInterpolate(A=@n33.RGB, B=@n36), B=@n37.RGB))
  n39 = TextureObjectParameter() {
   group=Wrap__PinkToons_Textures
   parameterName="PT_ColorTexture_01 - RGB Color Masks A=Alpha Mask"
   sortPriority=11
   texture=&T_BlacksRGB
  }
  n40 = TextureSample(
   UVs=MaterialFunctionCall(
    Caculated_Bounds=@n17
    Pre_Skinned_Normals=@n14
    Is Vehicle=@n32.R
    Is VehicleSide=@n32.R
    Is Mech=@n32.R
    Uvs_Offset=Add(A=@n10, B=@n38)
    Uvs_TileRate=@n13
   ) {materialFunction=&MF_Projection_Side_Normal}.UVs
   Tex=@n39
  )
  n41 = VectorParameter() {group=Wrap_PinkToons_Reactive, parameterName=PT_ColorTexture_01_ReactiveColor, defaultValue=<rgba 1 1 1 1>}
  n42 = MaterialFunctionCall(Color=@n5) {materialFunction=&SplitComponents}
  n43 = VectorParameter() {group=Wrap_PinkToons_UVs, parameterName=PT_ColorTexture_02_Panning_Weapon, defaultValue=<rgba 0 0 0 1>}
  n44 = VectorParameter() {group=Wrap_PinkToons_UVs, parameterName=PT_ColorTexture_02_Panning_Vehicle, defaultValue=<rgba 0 0 0 1>}
  n45 = VectorParameter() {group=Wrap_PinkToons_UVs, parameterName=PT_ColorTexture_02_Panning_VehicleSide, defaultValue=<rgba 0 0 0 1>}
  n46 = LinearInterpolate(A=@n44.RGB, B=@n45.RGB)
  n47 = VectorParameter() {group=Wrap_PinkToons_UVs, parameterName=PT_ColorTexture_02_Panning_Mech -----, defaultValue=<rgba 0 0 0 1>}
  n48 = Multiply(A=Time(), B=LinearInterpolate(A=LinearInterpolate(A=@n43.RGB, B=@n46), B=@n47.RGB))
  n49 = TextureObjectParameter() {
   group=Wrap__PinkToons_Textures
   parameterName="PT_ColorTexture_02 - RGB Color Masks A=Alpha Mask"
   sortPriority=12
   texture=&T_BlacksRGB
  }
  n50 = TextureSample(
   UVs=MaterialFunctionCall(
    Caculated_Bounds=@n17
    Pre_Skinned_Normals=@n14
    Is Vehicle=@n42.R
    Is VehicleSide=@n42.R
    Is Mech=@n42.R
    Uvs_Offset=Add(A=@n10, B=@n48)
    Uvs_TileRate=@n13
   ) {materialFunction=&MF_Projection_Side_Normal}.UVs
   Tex=@n49
  )
  n51 = VectorParameter() {group=Wrap_PinkToons_Reactive, parameterName=PT_ColorTexture_02_ReactiveColor, defaultValue=<rgba 0 0 0 1>}
  n52 = LinearInterpolate(A=Multiply(A=@n40.A, B=@n41.RGB), B=Multiply(A=@n50.A, B=@n51.RGB), Alpha=@n50.A)
  n53 = MaterialFunctionCall(Color=@n5) {materialFunction=&SplitComponents}
  n54 = VectorParameter() {group=Wrap_PinkToons_UVs, parameterName=PT_ColorTexture_03_Panning_Weapon, defaultValue=<rgba 0 0 0 1>}
  n55 = VectorParameter() {group=Wrap_PinkToons_UVs, parameterName=PT_ColorTexture_03_Panning_Vehicle, defaultValue=<rgba 0 0 0 1>}
  n56 = VectorParameter() {group=Wrap_PinkToons_UVs, parameterName=PT_ColorTexture_03_Panning_VehicleSide, defaultValue=<rgba 0 0 0 1>}
  n57 = LinearInterpolate(A=@n55.RGB, B=@n56.RGB)
  n58 = VectorParameter() {group=Wrap_PinkToons_UVs, parameterName=PT_ColorTexture_03_Panning_Mech -----, defaultValue=<rgba 0 0 0 1>}
  n59 = Multiply(A=Time(), B=LinearInterpolate(A=LinearInterpolate(A=@n54.RGB, B=@n57), B=@n58.RGB))
  n60 = TextureObjectParameter() {
   group=Wrap__PinkToons_Textures
   parameterName="PT_ColorTexture_03 - RGB Color Masks A=Alpha Mask"
   sortPriority=12
   texture=&T_BlacksRGB
  }
  n61 = TextureSample(
   UVs=MaterialFunctionCall(
    Caculated_Bounds=@n17
    Pre_Skinned_Normals=@n14
    Is Vehicle=@n53.R
    Is VehicleSide=@n53.R
    Is Mech=@n53.R
    Uvs_Offset=Add(A=@n10, B=@n59)
    Uvs_TileRate=@n13
   ) {materialFunction=&MF_Projection_Side_Normal}.UVs
   Tex=@n60
  )
  n62 = VectorParameter() {group=Wrap_PinkToons_Reactive, parameterName=PT_ColorTexture_03_ReactiveColor, defaultValue=<rgba 0 0 0 1>}
  n63 = Multiply(A=@n61.A, B=@n62.RGB)
  n64 = Divide(
   A=Subtract(
    A=Time()
    B=ScalarParameter() {defaultValue=-9999999, group=Wrap_PinkToons_Reactive, parameterName=TimeFrom-BP----------------, sortPriority=-600}
   )
   B=ScalarParameter() {defaultValue=1, group=Wrap_PinkToons_Reactive, parameterName=PT_AnimationDuration, sortPriority=33}
  )
  n65 = OneMinus(None=Power(
    Base=Saturate(None=@n64)
    Exp=ScalarParameter() {defaultValue=0.5, group=Wrap_PinkToons_Reactive, parameterName=PT_Curve_Amount, sortPriority=34}
   )) {desc=inverts curve from 0-1 to 1-0}
  n66 = ScalarParameter() {group=Wrap_PinkToons_Reactive, parameterName=PT_TestReactive, sortPriority=300}
  n67 = Constant() {r=0.85}
  n68 = Multiply(
   A=LinearInterpolate(B=LinearInterpolate(A=@n52, B=@n63, Alpha=@n61.A), Alpha=LinearInterpolate(A=@n65, B=@n66, Alpha=@n66))
   B=MaterialFunctionCall(Value_Morning=@n67, Value_Day=Constant() {r=1}, Value_Evening=@n67, Value_Night=Constant() {r=0.75}) {materialFunction=&Scalar_Time_of_Day}.Result
  )
  n69 = ScalarParameter() {defaultValue=0.5, group=Wrap_PinkToons, parameterName=PT_Specular}
  n70 = LinearInterpolate(A=LinearInterpolate(A=LinearInterpolate(B=@n40.A, Alpha=@n40.A), B=@n50.A, Alpha=@n50.A), B=@n61.A, Alpha=@n61.A)
  n71 = LinearInterpolate(A=@n69, B=ScalarParameter() {defaultValue=0.5, group=Wrap_PinkToons, parameterName=PT_Mask_Specular, sortPriority=52}, Alpha=@n70)
  n72 = TextureObjectParameter() {
   group=Wrap__PinkToons_Textures
   parameterName="PT_BaseColor - RGB=Color Masks A=Roughness"
   sortPriority=10
   texture=&T_BlacksRGB
  }
  n73 = TextureSample(UVs=@n18.UVs, Tex=@n72)
  n74 = LinearInterpolate(
   A=ScalarParameter() {group=Wrap_PinkToons, parameterName=PT_Roughness_Min}
   B=ScalarParameter() {defaultValue=1, group=Wrap_PinkToons, parameterName=PT_Roughness_Max}
   Alpha=ComponentMask(None=@n73.RGBA) {a=true}
  )
  n75 = LinearInterpolate(A=@n74, B=ScalarParameter() {defaultValue=0.5, group=Wrap_PinkToons, parameterName=PT_Mask_Roughness, sortPriority=52}, Alpha=@n70)
  n76 = VectorParameter() {group=Wrap_PinkToons, parameterName=PT_ColorTexture_01_ColorOverride, defaultValue=<rgba 0.0999 0.0152 0.9473 1>}
  n77 = LinearInterpolate(
   A=@n73.RGB
   B=LinearInterpolate(A=@n40.RGB, B=@n76.RGB, Alpha=ScalarParameter() {group=Wrap_PinkToons, parameterName=PT_ColorTexture_01__ColorOverride-----})
   Alpha=@n40.A
  )
  n78 = VectorParameter() {group=Wrap_PinkToons, parameterName=PT_ColorTexture_02_ColorOverride, defaultValue=<rgba 0.7835 0.016 0.807 1>}
  n79 = LinearInterpolate(A=@n50.RGB, B=@n78.RGB, Alpha=ScalarParameter() {group=Wrap_PinkToons, parameterName=PT_ColorTexture_02__ColorOverride})
  n80 = VectorParameter() {group=Wrap_PinkToons, parameterName=PT_ColorTexture_03_ColorOverride, defaultValue=<rgba 0.0999 0.3663 0.9823 1>}
  n81 = ScalarParameter() {group=Wrap_PinkToons, parameterName=PT_ColorTexture_03__ColorOverride}
  n82 = LinearInterpolate(
   A=LinearInterpolate(A=LinearInterpolate(A=@n77, B=@n79, Alpha=@n50.A), B=LinearInterpolate(A=@n61.RGB, B=@n80.RGB, Alpha=@n81), Alpha=@n61.A)
   B=@n73.RGB
   Alpha=@n30
  )
  n83 = ScalarParameter() {defaultValue=1, group=Wrap_PinkToons, parameterName=PT_DarkenBaseWhenEmissive-------, sortPriority=-20}
  n84 = OneMinus(None=Saturate(None=Multiply(A=@n83, B=@n68)))
  n85 = ScalarParameter() {group=Wrap_PinkToons, parameterName=PT_Metallic}
  n86 = LinearInterpolate(A=@n85, B=ScalarParameter() {group=Wrap_PinkToons, parameterName=PT_Mask_Metallic, sortPriority=52}, Alpha=@n70)
  n87 = SetMaterialAttributes!(
   MaterialAttributes=@n2.MaterialAttributes
   BaseColor=LinearInterpolate(A=Multiply(A=@n82, B=@n84), B=@n73.RGB, Alpha=@n30)
   Roughness=LinearInterpolate(
    A=LinearInterpolate(A=@n75, B=@n74, Alpha=@n30)
    B=@n74
    Alpha=ScalarParameter() {group=Wrap_PinkToons, parameterName=PT_Mask_RoughnessUseBase, sortPriority=55}
   )
   Normal=StaticSwitchParameter(True=@n2.Normal, False=ShadingPathSwitch(Default=@n31, Deferred=@n31, Mobile=@n2.Normal) {desc=Migrated from FeatureLevelSwitch}) {defaultValue=true, group=Wrap_PinkToons, parameterName=PT_Use No Normal, sortPriority=4}
   Specular=LinearInterpolate(
    A=LinearInterpolate(A=@n71, B=@n69, Alpha=@n30)
    B=@n69
    Alpha=ScalarParameter() {group=Wrap_PinkToons, parameterName=PT_Mask_SpecularUseBase, sortPriority=55}
   )
   Metallic=LinearInterpolate(
    A=LinearInterpolate(A=@n86, B=@n85, Alpha=@n30)
    B=@n85
    Alpha=ScalarParameter() {group=Wrap_PinkToons, parameterName=PT_Mask_MetallicUseBase-----, sortPriority=55}
   )
   EmissiveColor=Multiply(A=OneMinus(None=@n30), B=@n68)
  ) {attributeSetTypes [
    #2
    #3
    #1
    #4
    #5
    #6
   ]}
  n88 = FunctionInput() {bUsePreviewValueAsDefault=true, inputName="Customization Mask (_CM B Top left)", inputType=FunctionInput_Scalar, sortPriority=24}
  n89 = FunctionInput() {bUsePreviewValueAsDefault=true, inputName="Customization Mask 2 (_CM B Top Right)", inputType=FunctionInput_Scalar, sortPriority=25}
  n90 = Subtract(A=NamedRerouteUsage(), B=NamedRerouteUsage())
  n91 = BlendMaterialAttributes(
   A=@n2.MaterialAttributes
   B=@n87
   Alpha=LinearInterpolate(
    A=MaterialFunctionCall!("Customization Mask (_CM B Top left)"=@n88, "Customization Mask 2 (_CM B Top Right)"=@n89) {materialFunction=&/MF_Customization_AllMasking}.Result
    B=@n88
    Alpha=Saturate(None=@n90)
   )
  )
  n92 = GetMaterialAttributes(None=@n91) {attributeGetTypes [
    #2
    #3
    #5
    #4
    #6
   ]}
  n93 = ScalarParameter() {defaultValue=-0.391, group=Wrap_PinkToons, parameterName="PT_Saturation - Positive numbers = Desaturated", sortPriority=0}
  n94 = VectorParameter() {group=Wrap_PinkToons_Outline, parameterName=PT_Outline_Color-----, sortPriority=50, defaultValue=<rgba 0 0 0 1>}
  n95 = MaterialFunctionCall(Color=@n5) {materialFunction=&SplitComponents}
  n96 = LinearInterpolate(
   A=LinearInterpolate(
    A=ScalarParameter() {defaultValue=10, group=Wrap_PinkToons_Outline, parameterName=PT_Outline_Curvature_Mask A}
    B=ScalarParameter() {defaultValue=10, group=Wrap_PinkToons_Outline, parameterName=PT_Outline_Curvature_Mask A Vehicle}
    Alpha=@n95.R
   )
   B=LinearInterpolate(
    A=ScalarParameter() {defaultValue=-7, group=Wrap_PinkToons_Outline, parameterName=PT_Outline_Curvature_Mask B}
    B=ScalarParameter() {defaultValue=-7, group=Wrap_PinkToons_Outline, parameterName=PT_Outline_Curvature_Mask B Vehicle}
    Alpha=@n95.R
   )
   Alpha=FunctionInput(Preview=TextureSample() {samplerType=SAMPLERTYPE_Masks, texture=/Game/Weapons/FORT_Rifles/Textures/T_Scar_CM}.G) {bUsePreviewValueAsDefault=true, inputName=Curavture, inputType=FunctionInput_Scalar, sortPriority=24}
  )
  n97 = Fresnel(ExponentIn=ScalarParameter() {defaultValue=4.7243, group=Wrap_PinkToons_Outline, parameterName=PT_Outline_Exponent1}) {baseReflectFraction=0}
  n98 = LinearInterpolate(
   A=ScalarParameter() {defaultValue=-5.4358, group=Wrap_PinkToons_Outline, parameterName=PT_Outline_C1}
   B=ScalarParameter() {defaultValue=41.451, group=Wrap_PinkToons_Outline, parameterName=PT_Outline_C2}
   Alpha=Power(Base=OneMinus(None=@n97), Exp=ScalarParameter() {defaultValue=4.6385, group=Wrap_PinkToons_Outline, parameterName=PT_Outline_Exponent2})
  )
  n99 = Saturate(None=OneMinus(None=Multiply(A=@n96, B=@n98)))
  n100 = ScalarParameter() {defaultValue=1, group=Wrap_PinkToons_Outline, parameterName=PT_Outline_Roughness, sortPriority=50}
  n101 = ScalarParameter() {defaultValue=0.5, group=Wrap_PinkToons_Outline, parameterName=PT_Outline_Specular, sortPriority=50}
  n102 = ScalarParameter() {group=Wrap_PinkToons_Outline, parameterName=PT_Outline_Metallic, sortPriority=50}
  n103 = VectorParameter() {group=Wrap_PinkToons_Outline, parameterName=PT_Outline_EmissiveColor, sortPriority=50, defaultValue=<rgba 0 0 0 1>}
  n104 = SetMaterialAttributes!(
   MaterialAttributes=@n92.MaterialAttributes
   BaseColor=LinearInterpolate(A=Desaturation(None=@n92.BaseColor, Fraction=@n93), B=@n94.RGB, Alpha=@n99)
   Roughness=LinearInterpolate(
    A=LinearInterpolate(A=@n92.Roughness, B=@n100, Alpha=@n99)
    B=@n92.Roughness
    Alpha=ScalarParameter() {group=Wrap_PinkToons_Outline, parameterName=PT_Outline_RoughnessUseBase, sortPriority=50}
   )
   Specular=LinearInterpolate(
    A=LinearInterpolate(A=@n92.Specular, B=@n101, Alpha=@n99)
    B=@n92.Specular
    Alpha=ScalarParameter() {group=Wrap_PinkToons_Outline, parameterName=PT_Outline_SpecularUseBase, sortPriority=50}
   )
   Metallic=LinearInterpolate(
    A=LinearInterpolate(A=@n92.Metallic, B=@n102, Alpha=@n99)
    B=@n92.Metallic
    Alpha=ScalarParameter() {group=Wrap_PinkToons_Outline, parameterName=PT_Outline_MetallicUseBase, sortPriority=50}
   )
   EmissiveColor=LinearInterpolate(
    A=LinearInterpolate(A=@n92.EmissiveColor, B=@n103.RGB, Alpha=@n99)
    B=@n92.EmissiveColor
    Alpha=ScalarParameter() {group=Wrap_PinkToons_Outline, parameterName=PT_Outline_EmissiveUseBase, sortPriority=50}
   )
  ) {attributeSetTypes [
    #2
    #3
    #4
    #5
    #6
   ]}
  n105 = Subtract(A=NamedRerouteUsage(), B=NamedRerouteUsage())
  n106 = BlendMaterialAttributes(A=@n91, B=@n104, Alpha=LinearInterpolate(B=@n88, Alpha=Saturate(None=@n105)) {constA=1})
  FunctionOutput(None=StaticSwitchParameter(
    True=StaticSwitchParameter(True=@n106, False=@n91) {group=Wrap_PinkToons_Outline, parameterName=Use Simple Outline, sortPriority=0}
    False=@n1
   ) {desc=Side Projection Base - Emissive White GLow on weapon fire, group=Customization_Wrap, parameterName=Wrap_PinkToons})
  NamedRerouteDeclaration(None=MaterialFunctionCall(Float4=@n5) {materialFunction=&BreakOutFloat4Components}.A) {name=IsCharacter, nodeColor=<rgba 0.9144 1 0.0491 1>}
  n107 = FunctionInput(Preview=@n4) {bUsePreviewValueAsDefault=true, inputName=IsState_B, inputType=FunctionInput_Vector4}
  NamedRerouteDeclaration(None=ComponentMask(None=@n107) {a=true}) {name=isCharAccent, nodeColor=<rgba 0 0.9253 1 1>}
  NamedRerouteDeclaration(None=ComponentMask(None=@n107) {g=true}) {name=IsPickaxe, nodeColor=<rgba 1 0 0.5625 1>}
  NamedRerouteDeclaration(None=ComponentMask(None=@n107) {r=true}) {name=IsBackpack, nodeColor=<rgba 0.2656 0 1 1>}
  NamedRerouteUsage()
  NamedRerouteUsage()
 ]}
