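# Material function for the PlasmaCore weapon wrap, drawing an animated glowing plasma core over the incoming material behind a static switch.
# SafeZone virtual-plane UVs become radial coordinates driving two scrolling, noise-warped, flickering plasma layers and core masks, colored through curve-atlas gradients.
# Audio collection parameters and a timestamp-driven reactive fade provide speed, brightness and opacity modulation.
# Top/side projection supplies roughness and warp channels; input masks blend wrap and accent materials.
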MaterialFunction /Game/Weapons/WeaponSkins/Wraps/MaterialFunctions/MF_Wrap_PlasmaCore
graph {nodes [
  n1 = Add(A=NamedRerouteUsage()) {constB=-0.5}
  n2 = MaterialFunctionCall!(
   Camera _pos
   Axis
   Scale
   Scale_Multplyer
   Object-SafeZone-Virtual-Pivot=NamedRerouteUsage()
   Object-SafeZone-Virtual-CH-Switch=NamedRerouteUsage()
   SinglePlane Normal Axis
   SinglePlane X Axis
   PlaneCenter_distortion
  ) {materialFunction=&/MF_SafeZone_VirtualPlane}
  n3 = Add(A=@n2.Uvs) {constB=-0.5}
  Add(
   A=Multiply(A=ScalarParameter() {defaultValue=0.5, group=Wrap_PlasmaCore_Back, parameterName=PlasC_Back_Cruvature_warp_amount}, B=@n1) {constB=2}
   B=@n3
  )
  n4 = MaterialFunctionCall(Vector or UVs=@n3) {materialFunction=&VectorToRadialValue}
  Add(A=NamedRerouteUsage(), B=@n4.Radial Coordinates)
  n5 = OneMinus(None=NamedRerouteUsage())
  n6 = CurveAtlasRowParameter(CurveTime=@n5) {
   atlas=/Game/Weapons/WeaponSkins/Wraps/Textures/CruveAtlas_Wraps_01
   curve=/Game/Weapons/WeaponSkins/Wraps/Textures/0_ColorCurves/CC_Plasma_A
   defaultValue=125
   group=Wrap_PlasmaCore_Core
   parameterName=PlasC_CC_Core
  }
  n7 = MaterialFunctionCall!(
   Index=ScalarParameter() {defaultValue=125, group=Wrap_PlasmaCore_Core, parameterName="PlasC_CC_Core  Index"}
   CCinput=@n6
   CCinput_Alpha
   UV=@n5
  ) {materialFunction=&/MF_Global_CC_Tex}
  n8 = VectorParameter() {
   group=Wrap_PlasmaCore_Core
   parameterName=PlasC_Core_Edge_Msk_Inner_Fallof
   channelNames {a="", b="", g=Max, r=Min}
   defaultValue=<rgba 0.4 0.5 0 0>
  }
  n9 = NamedRerouteUsage()
  n10 = VectorParameter() {
   group=Wrap_PlasmaCore_Core
   parameterName=PlasC_Core_Edge_Msk_Edge_Msk_Fallof
   channelNames {a="", b=Power, g=Max, r=Min}
   defaultValue=<rgba 1 0.5 2 0>
  }
  n11 = Power(Base=SmoothStep(Min=@n10.Min, Max=@n10.Min, Value=@n9) {constMax=0.5, constMin=1}, Exp=@n10.Power)
  n12 = VectorParameter() {
   group=Wrap_PlasmaCore_Core
   parameterName=PlasC_Core_Edge_Msk_Edge_Blend_Fallof
   channelNames {a="", b="", g=Max, r=Min}
   defaultValue=<rgba 0.5 0.41 0 0>
  }
  n13 = Add(A=SmoothStep(Min=@n12.Min, Max=@n12.Min, Value=@n9) {constMax=0.41, constMin=0.5}, B=NamedRerouteUsage()) {constB=0}
  n14 = CurveAtlasRowParameter(CurveTime=@n13) {
   atlas=&CruveAtlas_Wraps_01
   curve=/Game/Weapons/WeaponSkins/Wraps/Textures/0_ColorCurves/CC_Plasma_C
   defaultValue=127
   group=Wrap_PlasmaCore_Plasma
   parameterName=PlasC_CC_Edge
  }
  n15 = Multiply(
   A=MaterialFunctionCall(Index=ScalarParameter() {defaultValue=127, group=Wrap_PlasmaCore_Plasma, parameterName="PlasC_CC_Edge  Index"}, CCinput=@n14, UV=@n13) {materialFunction=&MF_Global_CC_Tex}.3V
   B=NamedRerouteUsage()
  )
  n16 = Multiply(A=SmoothStep(Min=@n8.Min, Max=@n8.Min, Value=@n9) {constMax=0.5, constMin=0.4}, B=Add(A=Multiply(A=@n11, B=@n15), B=NamedRerouteUsage()))
  n17 = Add(A=Max(A=Add(A=@n7.3V, B=@n16), B=NamedRerouteUsage()), B=NamedRerouteUsage())
  Add(A=Max(A=NamedRerouteUsage(), B=NamedRerouteUsage()), B=@n17)
  n18 = Add(
   A=Add(A=NamedRerouteUsage(), B=NamedRerouteUsage()) {constB=0}
   B=Multiply(A=Time(), B=ScalarParameter() {defaultValue=-0.015, group=Wrap_PlasmaCore_Plasma, parameterName=PlasC_Plasma_B_Speed})
  )
  AppendVector(A=ComponentMask(None=@n18) {r=true}, B=Constant())
  Constant2Vector() {g=0.5, r=0.025}
  Constant()
  Constant() {r=1}
  Constant()
  FunctionInput(Preview=SetMaterialAttributes!(
    MaterialAttributes
    Normal=TextureSample() {samplerType=SAMPLERTYPE_Normal, texture=/Game/Weapons/FORT_Rifles/T_SCAR_N}.RGB
    BaseColor=Constant3Vector() {constant=<rgba 1 0 0 0>}
   ) {attributeSetTypes [#1, #2]}) {inputName=Material _Root, inputType=FunctionInput_MaterialAttributes}
  n19 = TextureSample() {samplerType=SAMPLERTYPE_Masks, texture=/Game/Weapons/FORT_Rifles/Textures/T_Scar_CM}
  n20 = FunctionInput(Preview=@n19.R) {bUsePreviewValueAsDefault=true, inputName=Mask Primary, inputType=FunctionInput_Scalar, sortPriority=24, previewValue=<1 0 0 1>}
  n21 = FunctionInput(Preview=@n19.B) {bUsePreviewValueAsDefault=true, inputName=Mask Secondary, inputType=FunctionInput_Scalar, sortPriority=24, previewValue=<1 0 0 1>}
  n22 = BlendMaterialAttributes(A=NamedRerouteUsage(), B=NamedRerouteUsage(), Alpha=Saturate(None=Add(A=@n20, B=@n21))) {vertexAttributeBlendType=UseA}
  n23 = StaticSwitchParameter(True=NamedRerouteUsage(), False=NamedRerouteUsage()) {group=Side Projection, parameterName=Use Side Projection As Secondary, sortPriority=-100}
  n24 = NamedRerouteDeclaration(None=@n21) {name=Msk_Accent, nodeColor=<rgba 0 1 0.7656 1>}
  n25 = FunctionInput() {inputName=Material_1, inputType=FunctionInput_MaterialAttributes}
  FunctionOutput(None=StaticSwitchParameter(True=BlendMaterialAttributes(A=@n22, B=@n23, Alpha=@n24) {vertexAttributeBlendType=UseA}, False=@n25) {group=Customization_wrap, parameterName=Wrap_PlasmaCore})
  MaterialFunctionCall(Float2=NamedRerouteUsage()) {materialFunction=&BreakOutFloat2Components}
  n26 = CollectionParameter() {collection=&FortniteMaterialParameters, parameterName=Emote_AudioLow}
  n27 = CollectionParameter() {collection=&FortniteMaterialParameters, parameterName=Emote_AudioHigh}
  n28 = ScalarParameter() {defaultValue=1, group=Wrap_PlasmaCore_Audio, parameterName=PlasC_EventAudioRemoteMultiplier}
  n29 = CollectionParameter() {collection=&FortniteMaterialParameters, parameterName=EventAudio_Amplitude}
  n30 = CollectionParameter() {collection=&FortniteMaterialParameters, parameterName=EventAudio_Spectrum}
  n31 = MaterialFunctionCall(Float4=@n30) {materialFunction=&BreakOutFloat4Components}
  Max(
   A=Multiply(A=Add(A=@n26, B=@n27), B=@n28) {constB=0.5}
   B=Multiply(A=Add(A=@n29, B=@n31.A), B=ScalarParameter() {defaultValue=1, group=Wrap_PlasmaCore_Audio, parameterName=PlasC_EventAudioMultiplier})
  )
  n32 = Subtract(
   A=Time()
   B=Add(
    A=ScalarParameter() {defaultValue=-9999999, group=Wrap_Reactive, parameterName=TimeFrom-BP----------------, sortPriority=600}
    B=ScalarParameter() {group=Wrap_PlasmaCore_Reactive, parameterName=PlasC_Reactive_Delay}
   )
  )
  n33 = ScalarParameter() {defaultValue=0.25, group=Wrap_PlasmaCore_Reactive, parameterName=PlasC_Reactive_Duration}
  n34 = Power(
   Base=Saturate(None=Divide(A=@n32, B=@n33))
   Exp=ScalarParameter() {defaultValue=0.25, group=Wrap_PlasmaCore_Reactive, parameterName=PlasC_Reactive_Falloff}
  )
  n35 = NamedRerouteUsage()
  Max(A=@n34, B=@n35)
  n36 = CollectionParameter() {collection=&FortniteMaterialParameters, parameterName=Emote_AudioLow}
  Min(A=Multiply(A=@n36, B=@n28)) {constB=30}
  n37 = MaterialFunctionCall() {materialFunction=&ObjectScale}
  n38 = NamedRerouteUsage()
  n39 = ComponentMask(None=@n38) {b=true, g=true, r=true}
  n40 = Constant()
  n41 = VectorParameter() {group=Wrap_PlasmaCore_Uvs, parameterName=PlasC_Core_offset_Vehicle}
  n42 = MaterialFunctionCall(Float4=NamedRerouteUsage()) {materialFunction=&BreakOutFloat4Components}
  n43 = LinearInterpolate(
   A=LinearInterpolate(
    A=LinearInterpolate(A=@n40, B=@n41.RGB, Alpha=@n42.R)
    B=VectorParameter() {group=Wrap_PlasmaCore_Uvs, parameterName=PlasC_Core_offset_Mech, defaultValue=<rgba 0 0 0 1>}.RGB
    Alpha=@n42.B
   )
   B=@n40
   Alpha=@n42.A
  )
  n44 = LinearInterpolate(
   A=TransformPosition(None=Add(A=@n39, B=@n43)) {transformType=TRANSFORMPOSSOURCE_World}
   B=ComponentMask(None=@n38) {b=true, g=true, r=true}
   Alpha=NamedRerouteUsage()
  )
  n45 = ComponentMask(None=Add(A=@n38, B=LinearInterpolate(A=@n40, B=@n41.A, Alpha=@n42.R))) {a=true}
  n46 = PreSkinnedLocalBounds()
  n47 = MaterialFunctionCall(Float3=Subtract(A=@n46.Max, B=@n46.Max)) {materialFunction=&BreakOutFloat3Components}
  n48 = Max(A=Max(A=@n47.R, B=@n47.R), B=@n47.B)
  Multiply(
   A=Multiply(
    A=Divide(B=@n37.Scale Z)
    B=MaterialFunctionCall!(Center Point World Position=@n44) {materialFunction=/Engine/Functions/Engine_MaterialFunctions02/UVs/WorldSpaceAlignedScreenCoordinates}.Offset and Scaled UVS
   )
   B=Divide(A=Constant() {r=-2500}, B=Multiply(A=@n45, B=@n48))
  )
  n49 = Multiply(A=Add(A=NamedRerouteUsage()) {constB=-0.5}) {constB=5}
  n50 = Add(A=Saturate(None=Multiply(A=@n49) {constB=-1}), B=Saturate(None=@n49)) {constB=0}
  n51 = NamedRerouteUsage()
  n52 = Add(
   A=Fresnel(ExponentIn=ScalarParameter() {defaultValue=3, group=Wrap_PlasmaCore, parameterName=PlasC_Fresnel_Exp}) {baseReflectFraction=0}
   B=Multiply(A=@n51, B=ScalarParameter() {defaultValue=0.5, group=Wrap_PlasmaCore, parameterName=PlasC_Fing_Opacity})
  )
  n53 = Saturate(None=Add(A=Saturate(None=@n50), B=Saturate(None=@n52)))
  n54 = OneMinus(None=@n53)
  Multiply(A=@n54)
  Multiply(A=MaterialFunctionCall(Value_Morning=Constant() {r=0.5}, Value_Day=Constant() {r=1}, Value_Evening=Constant() {r=0.5}, Value_Night=Constant() {r=0.25}) {materialFunction=&Scalar_Time_of_Day}.Result)
  Multiply(A=TextureSample(UVs=@n2.Uvs) {texture=/Game/Weapons/WeaponSkins/Wraps/Textures/UVCheaker_Clamped}.RGB)
  n55 = VectorParameter() {group=Wrap_PlasmaCore_Accent, parameterName=PlasC_Accent_BaseColor, defaultValue=<rgba 0.0156 0.0156 0.0156 1>}
  n56 = SetMaterialAttributes!(
   MaterialAttributes=NamedRerouteUsage()
   BaseColor=@n55.RGB
   Roughness=ScalarParameter() {defaultValue=0.5, group=Wrap_PlasmaCore_Accent, parameterName=PlasC_Accent_Roughness}
   Specular=ScalarParameter() {defaultValue=0.5, group=Wrap_PlasmaCore_Accent, parameterName=PlasC_Accent_Spec}
   Metallic=ScalarParameter() {group=Wrap_PlasmaCore_Accent, parameterName=PlasC_Accent_Metalic}
   Normal
   EmissiveColor
  ) {attributeSetTypes [
    #2
    #3
    #4
    #5
    #1
    #6
   ]}
  NamedRerouteDeclaration(None=@n56) {name=Mat_Accent, nodeColor=<rgba 0 0.3906 1 1>}
  n57 = FunctionInput(Preview=@n19.G) {bUsePreviewValueAsDefault=true, inputName=Mask Crevise, inputType=FunctionInput_Scalar, sortPriority=24, previewValue=<1 0 0 1>}
  NamedRerouteDeclaration(None=@n57) {name=Msk_Curvature, nodeColor=<rgba 0 1 0.7422 1>}
  n58 = GetMaterialAttributes(None=@n25) {attributeGetTypes [
    #2
    #3
    #1
    #5
    #6
   ]}
  NamedRerouteDeclaration(None=@n58.MaterialAttributes) {name=Root_Material, nodeColor=<rgba 0.7812 1 0 1>}
  n59 = FunctionInput(Preview=Constant4Vector() {constant=<rgba 0 0 5 -1>}.RGBA) {bUsePreviewValueAsDefault=true, inputName=Object-SafeZone-Virtual-Pivot, inputType=FunctionInput_Vector4, sortPriority=105}
  NamedRerouteDeclaration(None=@n59) {name=SafeZone VP, nodeColor=<rgba 0.5 1 0 1>}
  n60 = FunctionInput(Preview=Constant()) {bUsePreviewValueAsDefault=true, inputName=Object-SafeZone-VP-Character-PreView-Switch, inputType=FunctionInput_Scalar, sortPriority=105}
  NamedRerouteDeclaration(None=@n60) {name=SafeZone Character-Switch, nodeColor=<rgba 1 0.9375 0 1>}
  n61 = Multiply(
   A=NamedRerouteUsage()
   B=ComponentMask(None=VectorParameter() {group=Wrap_PlasmaCore_Warp, parameterName=PlasC_Warp_TileRate_Core, defaultValue=<rgba 3 1 0 1>}.RGB) {g=true, r=true}
  )
  n62 = ComponentMask(None=Multiply(A=Time(), B=VectorParameter() {group=Wrap_PlasmaCore_Warp, parameterName=PlasC_Warp_Speed_Core, defaultValue=<rgba 0.05 -0.25 0 1>}.RGB)) {g=true, r=true}
  n63 = Add(A=TextureSample!(UVs=Add(A=@n61, B=@n62), Tex=NamedRerouteUsage(), Level=NamedRerouteUsage(), Apply View MipBias) {mipValueMode=TMVM_MipLevel, samplerType=SAMPLERTYPE_LinearGrayscale}.A) {constB=-0.5}
  n64 = ScalarParameter() {defaultValue=0.05, group=Wrap_PlasmaCore_Warp, parameterName=PlasC_warp_amount_Core}
  NamedRerouteDeclaration(None=Multiply(A=NamedRerouteUsage(), B=Multiply(A=@n63, B=@n64))) {name=Warp_Core, nodeColor=<rgba 0.5469 0 1 1>}
  n65 = ScalarParameter() {defaultValue=15, group=Wrap_PlasmaCore_Core, parameterName=PlasC_Core_Size_A}
  n66 = MaterialFunctionCall(Vector 2=@n3) {materialFunction=/Engine/Functions/Engine_MaterialFunctions02/Utility/VectorLength}
  NamedRerouteDeclaration(None=SmoothStep(Value=Multiply(A=@n65, B=@n66.V2 Length)) {constMax=0.4, constMin=0.2}) {name=Msk_Core_inner, nodeColor=<rgba 1 0.5859 0 1>}
  n67 = VectorParameter() {group=Wrap_PlasmaCore_Warp, parameterName=PlasC_Warp_Plasma_A_TileRate, defaultValue=<rgba 1 3 0 1>}
  n68 = VectorParameter() {group=Wrap_PlasmaCore_Warp, parameterName=PlasC_Warp_Plasma_A_TileRate_Falloff, defaultValue=<rgba 1 3 0 1>}
  n69 = Multiply(
   A=NamedRerouteUsage()
   B=LinearInterpolate(A=ComponentMask(None=@n67.RGB) {g=true, r=true}, B=ComponentMask(None=@n68.RGB) {g=true, r=true}, Alpha=NamedRerouteUsage())
  )
  n70 = ComponentMask(None=Multiply(A=Time(), B=VectorParameter() {group=Wrap_PlasmaCore_Warp, parameterName=PlasC_Warp_Plasma_A_Speed, defaultValue=<rgba 0.1 0 0 1>}.RGB)) {g=true, r=true}
  n71 = Add(
   A=NamedRerouteUsage()
   B=Multiply(A=NamedRerouteUsage(), B=ScalarParameter() {defaultValue=0.5, group=Wrap_PlasmaCore_Audio, parameterName=PlasC_Audio_speed_mult_A})
  )
  n72 = Constant()
  n73 = TextureSample(UVs=Add(A=Add(A=@n69, B=@n70), B=AppendVector(A=@n71, B=@n72)), Tex=NamedRerouteUsage(), Level=NamedRerouteUsage()) {mipValueMode=TMVM_MipLevel, samplerType=SAMPLERTYPE_LinearGrayscale}
  n74 = ScalarParameter() {defaultValue=0.1, group=Wrap_PlasmaCore_Warp, parameterName=PlasC_Warp_Plasma_A_amount}
  n75 = ScalarParameter() {defaultValue=0.1, group=Wrap_PlasmaCore_Warp, parameterName=PlasC_Warp_Plasma_A_amount_Falloff}
  n76 = NamedRerouteUsage()
  NamedRerouteDeclaration(None=Multiply(A=Add(A=@n73.A) {constB=-0.5}, B=LinearInterpolate(A=@n74, B=@n75, Alpha=@n76))) {name=Warp_Plasma_A, nodeColor=<rgba 0 0.2031 1 1>}
  n77 = Add(
   A=Add(A=NamedRerouteUsage(), B=NamedRerouteUsage()) {constB=0}
   B=Multiply(A=Time(), B=ScalarParameter() {defaultValue=0.01, group=Wrap_PlasmaCore_Plasma, parameterName=PlasC_Plasma_A_Speed})
  )
  n78 = Multiply(A=@n77, B=ScalarParameter() {defaultValue=20, group=Wrap_PlasmaCore_Plasma, parameterName=PlasC_Plasma_A_TileRate})
  n79 = Abs(None=Subtract(A=Frac(None=@n78)) {constB=0.5})
  n80 = VectorParameter() {group=Wrap_PlasmaCore_Plasma, parameterName=PlasC_Plasma_Flicker_A_Offset, defaultValue=<rgba 0.025 0 0 1>}
  n81 = VectorParameter() {group=Wrap_PlasmaCore_Plasma, parameterName=PlasC_Plasma_Flicker_A_TileRate, defaultValue=<rgba 1 0 0 1>}
  n82 = Add(A=ComponentMask(None=@n80.RGB) {g=true, r=true}, B=Multiply(A=@n77, B=ComponentMask(None=@n81.RGB) {g=true, r=true}))
  n83 = VectorParameter() {group=Wrap_PlasmaCore_Plasma, parameterName=PlasC_Plasma_Flicker_A_Speed, defaultValue=<rgba 0 0.1 0 1>}
  n84 = ComponentMask(None=Multiply(A=Time(), B=@n83.RGB)) {g=true, r=true}
  n85 = Add(
   A=TextureSample(UVs=Add(A=@n82, B=@n84), Tex=NamedRerouteUsage(), Level=NamedRerouteUsage()) {mipValueMode=TMVM_MipLevel}.R
   B=NamedRerouteUsage()
  )
  n86 = Multiply(A=NamedRerouteUsage(), B=Multiply(A=ComponentMask(None=@n79) {r=true}, B=Saturate(None=@n85)))
  n87 = SmoothStep(Value=@n86) {constMin=0.1}
  NamedRerouteDeclaration(None=Saturate(None=Add(A=@n87) {constB=0.25})) {name=Plasma_A_Msk, nodeColor=<rgba 0.3125 0 1 1>}
  n88 = NamedRerouteUsage()
  n89 = ScalarParameter() {group=Wrap_PlasmaCore, parameterName="PlasC_BaseColor use root 1= Active"}
  n90 = VectorParameter() {group=Wrap_PlasmaCore, parameterName=PlasC_BaseColor_A, defaultValue=<rgba 0.002 0.002 0.0025 1>}
  n91 = VectorParameter() {group=Wrap_PlasmaCore, parameterName=PlasC_BaseColor_B, defaultValue=<rgba 0.0766 0.0863 0.1 1>}
  n92 = LinearInterpolate(
   A=ScalarParameter() {group=Wrap_PlasmaCore, parameterName=PlasC_Rougness_High}
   B=ScalarParameter() {defaultValue=1, group=Wrap_PlasmaCore, parameterName=PlasC_Rougness_Low}
   Alpha=@n51
  )
  n93 = NamedRerouteUsage()
  n94 = Saturate(None=Multiply(A=Add(A=@n93) {constB=-0.5}) {constB=2})
  n95 = Multiply(A=@n94, B=ScalarParameter() {defaultValue=1, group=Wrap_PlasmaCore, parameterName=PlasC_Rougness_Curvature})
  n96 = Saturate(None=Add(A=@n54, B=ScalarParameter() {group=Wrap_PlasmaCore, parameterName="PlasC_Emissive Fren\\curvature Msk 1=off"}))
  n97 = VectorParameter() {group=Wrap_PlasmaCore, parameterName=PlasC_Edge_Color, defaultValue=<rgba 0.015 0.015 0.015 1>}
  n98 = NamedRerouteUsage()
  n99 = NamedRerouteUsage()
  n100 = LinearInterpolate(
   A=ScalarParameter() {defaultValue=1, group=Wrap_PlasmaCore_Audio, parameterName=PlasC_Audio_Brightness_Min}
   B=ScalarParameter() {defaultValue=5, group=Wrap_PlasmaCore_Audio, parameterName=PlasC_Audio_Brightness_Max}
   Alpha=Saturate(None=@n99)
  )
  n101 = Add(A=Multiply(A=@n17, B=Max(A=@n98, B=@n100)), B=NamedRerouteUsage())
  n102 = Multiply(A=@n101, B=ScalarParameter() {defaultValue=3, group=Wrap_PlasmaCore, parameterName=PlasC_Edge_Brightness})
  n103 = NamedRerouteUsage()
  n104 = Saturate(None=Multiply(A=Add(A=@n103) {constB=-0.5}) {constB=2})
  n105 = Add(
   A=LinearInterpolate(B=NamedRerouteUsage(), Alpha=ScalarParameter() {group=Wrap_PlasmaCore, parameterName="PlasC_Root_Emissive  1=Active"})
   B=Add(A=Multiply(A=@n94, B=@n97.RGB), B=LinearInterpolate(A=@n101, B=@n102, Alpha=@n104))
  )
  n106 = SetMaterialAttributes!(
   MaterialAttributes=NamedRerouteUsage()
   BaseColor=Add(A=LinearInterpolate(B=@n88, Alpha=@n89), B=LinearInterpolate(A=@n90.RGB, B=@n91.RGB, Alpha=@n53))
   Metallic=ScalarParameter() {defaultValue=0.5, group=Wrap_PlasmaCore, parameterName=PlasC_Metalic}
   Specular=ScalarParameter() {defaultValue=0.3, group=Wrap_PlasmaCore, parameterName=PlasC_Spec}
   Roughness=Saturate(None=Add(A=@n92, B=@n95))
   EmissiveColor=Clamp(None=Multiply(A=@n96, B=@n105), Max=ScalarParameter() {defaultValue=20, group=Wrap_PlasmaCore, parameterName=PlasC_Emissive Max Brightness})
  ) {attributeSetTypes [
    #2
    #5
    #4
    #3
    #6
   ]}
  n107 = GetMaterialAttributes(None=@n106) {attributeGetTypes [#2, #6]}
  n108 = ScalarParameter() {group=Wrap_PlasmaCore, parameterName="PlasC_Emissive ---> BaseColor"}
  n109 = ScalarParameter() {defaultValue=1, group=Wrap_PlasmaCore, parameterName=PlasC_Emissive - Brightness}
  NamedRerouteDeclaration(None=SetMaterialAttributes!(
    MaterialAttributes=@n107.MaterialAttributes
    BaseColor=LinearInterpolate(A=@n107.BaseColor, B=@n107.BaseColor, Alpha=@n108) {constB=0}
    EmissiveColor=Multiply(A=@n107.EmissiveColor, B=@n109)
   ) {attributeSetTypes [#2, #6]}) {name=Mat_Base, nodeColor=<rgba 0 0.3906 1 1>}
  n110 = NamedRerouteUsage()
  n111 = Saturate(None=Multiply(A=Add(A=@n110) {constB=-0.5}))
  n112 = Multiply(A=@n111) {constB=2}
  n113 = SmoothStep(Value=Add(A=NamedRerouteUsage(), B=@n111)) {constMax=0.75, constMin=0.45}
  n114 = VectorParameter() {group=Wrap_PlasmaCore_Plasma, parameterName=PlasC_Plasma_Edge_Color, defaultValue=<rgba 0.243 0.0458 0.5 1>}
  NamedRerouteDeclaration(None=Multiply(A=SmoothStep(Value=@n112) {constMax=0.25}, B=Multiply(A=@n113, B=@n114.RGB))) {name=Plasma_A_Cure, nodeColor=<rgba 1 0 0.1406 1>}
  NamedRerouteDeclaration(None=@n4.Radial Coordinates) {name=UVs, nodeColor=<rgba 0.9453 1 0 1>}
  n115 = NamedRerouteDeclaration(None=Power(Base=Multiply(A=@n86) {constB=2}) {constExponent=3}) {name=Plasma_A, nodeColor=<rgba 0.3125 0 1 1>}
  n116 = CurveAtlasRowParameter(CurveTime=@n115) {
   atlas=&CruveAtlas_Wraps_01
   curve=/Game/Weapons/WeaponSkins/Wraps/Textures/0_ColorCurves/CC_Plasma_B
   defaultValue=126
   group=Wrap_PlasmaCore_Plasma
   parameterName=PlasC_CC_Plasma
  }
  NamedRerouteDeclaration(None=MaterialFunctionCall(Index=ScalarParameter() {defaultValue=126, group=Wrap_PlasmaCore_Plasma, parameterName="PlasC_CC_Plasma  Index"}, CCinput=@n116, UV=@n115) {materialFunction=&MF_Global_CC_Tex}.3V) {name=Plasma_A_CC, nodeColor=<rgba 1 0 0.1406 1>}
  n117 = VectorParameter() {group=Wrap_PlasmaCore_Warp, parameterName=PlasC_Warp_Plasma_B_TileRate, defaultValue=<rgba 1 2.5 0 1>}
  n118 = VectorParameter() {group=Wrap_PlasmaCore_Warp, parameterName=PlasC_Warp_Plasma_B_TileRate_Falloff, defaultValue=<rgba 1 2.5 0 1>}
  n119 = Multiply(
   A=NamedRerouteUsage()
   B=LinearInterpolate(A=ComponentMask(None=@n117.RGB) {g=true, r=true}, B=ComponentMask(None=@n118.RGB) {g=true, r=true}, Alpha=NamedRerouteUsage())
  )
  n120 = VectorParameter() {group=Wrap_PlasmaCore_Warp, parameterName=PlasC_Warp_Plasma_B_Speed, defaultValue=<rgba 0.133 -0.1 0 1>}
  n121 = ComponentMask(None=Multiply(A=Time(), B=@n120.RGB)) {g=true, r=true}
  n122 = Add(
   A=NamedRerouteUsage()
   B=Multiply(A=NamedRerouteUsage(), B=ScalarParameter() {defaultValue=0.5, group=Wrap_PlasmaCore_Audio, parameterName=PlasC_Audio_speed_mult_B})
  )
  n123 = Constant()
  n124 = TextureSample(UVs=Add(A=Add(A=@n119, B=@n121), B=AppendVector(A=@n122, B=@n123)), Tex=NamedRerouteUsage(), Level=NamedRerouteUsage()) {mipValueMode=TMVM_MipLevel, samplerType=SAMPLERTYPE_LinearGrayscale}
  n125 = ScalarParameter() {defaultValue=0.1, group=Wrap_PlasmaCore_Warp, parameterName=PlasC_Warp_Plasma_B_amount}
  n126 = ScalarParameter() {defaultValue=0.1, group=Wrap_PlasmaCore_Warp, parameterName=PlasC_Warp_Plasma_B_amount_falloff}
  n127 = NamedRerouteUsage()
  NamedRerouteDeclaration(None=Multiply(A=Add(A=@n124.A) {constB=-0.5}, B=LinearInterpolate(A=@n125, B=@n126, Alpha=@n127))) {name=Warp_Plasma_B, nodeColor=<rgba 0 0.2031 1 1>}
  n128 = Multiply(A=@n18, B=ScalarParameter() {defaultValue=20, group=Wrap_PlasmaCore_Plasma, parameterName=PlasC_Plasma_B_TileRate})
  n129 = Abs(None=Subtract(A=Frac(None=@n128)) {constB=0.5})
  n130 = VectorParameter() {group=Wrap_PlasmaCore_Plasma, parameterName=PlasC_Plasma_Flicker_B_Offset, defaultValue=<rgba 0.025 0.25 0 1>}
  n131 = VectorParameter() {group=Wrap_PlasmaCore_Plasma, parameterName=PlasC_Plasma_Flicker_B_TileRate, defaultValue=<rgba 1 0 0 1>}
  n132 = Add(A=ComponentMask(None=@n130.RGB) {g=true, r=true}, B=Multiply(A=@n18, B=ComponentMask(None=@n131.RGB) {g=true, r=true})) {constB=0.025}
  n133 = VectorParameter() {group=Wrap_PlasmaCore_Plasma, parameterName=PlasC_Plasma_Flicker_B_Speed, defaultValue=<rgba 0 -0.1 0 1>}
  n134 = ComponentMask(None=Multiply(A=Time(), B=@n133.RGB)) {g=true, r=true}
  n135 = Add(
   A=TextureSample(UVs=Add(A=@n132, B=@n134), Tex=NamedRerouteUsage(), Level=NamedRerouteUsage()) {mipValueMode=TMVM_MipLevel}.R
   B=NamedRerouteUsage()
  )
  n136 = Multiply(A=NamedRerouteUsage(), B=Multiply(A=ComponentMask(None=@n129) {r=true}, B=Saturate(None=@n135)))
  n137 = SmoothStep(Value=@n136) {constMin=0.1}
  NamedRerouteDeclaration(None=Saturate(None=Add(A=@n137) {constB=0.25})) {name=Plasma_B_Msk, nodeColor=<rgba 0.3125 0 1 1>}
  n138 = NamedRerouteDeclaration(None=Power(Base=Multiply(A=@n136) {constB=2}) {constExponent=3}) {name=Plasma_B, nodeColor=<rgba 0.3125 0 1 1>}
  n139 = CurveAtlasRowParameter(CurveTime=@n138) {
   atlas=&CruveAtlas_Wraps_01
   curve=&CC_Plasma_B
   defaultValue=126
   group=Wrap_PlasmaCore_Plasma
   parameterName=PlasC_CC_Plasma
  }
  NamedRerouteDeclaration(None=MaterialFunctionCall(Index=ScalarParameter() {defaultValue=126, group=Wrap_PlasmaCore_Plasma, parameterName="PlasC_CC_Plasma  Index"}, CCinput=@n139, UV=@n138) {materialFunction=&MF_Global_CC_Tex}.3V) {name=Plasma_B_CC, nodeColor=<rgba 1 0 0.1406 1>}
  n140 = NamedRerouteUsage()
  n141 = Saturate(None=Multiply(A=Add(A=@n140) {constB=-0.5}))
  n142 = Multiply(A=@n141) {constB=2}
  n143 = SmoothStep(Value=Add(A=NamedRerouteUsage(), B=@n141)) {constMax=0.75, constMin=0.45}
  n144 = VectorParameter() {group=Wrap_PlasmaCore_Plasma, parameterName=PlasC_Plasma_Edge_Color, defaultValue=<rgba 0.243 0.0458 0.5 1>}
  NamedRerouteDeclaration(None=Multiply(A=SmoothStep(Value=@n142) {constMax=0.25}, B=Multiply(A=@n143, B=@n144.RGB))) {name=Plasma_B_Cure, nodeColor=<rgba 1 0 0.1406 1>}
  n145 = Multiply(A=ScalarParameter() {defaultValue=0.05, group=Wrap_PlasmaCore, parameterName=PlasC_Cruvature_warp_amount}, B=@n1) {constB=2}
  n146 = Multiply(A=@n65, B=Add(A=@n145, B=@n3)) {constA=1}
  n147 = NamedRerouteUsage()
  NamedRerouteDeclaration(None=MaterialFunctionCall(Vector 2=Add(A=@n146, B=@n147)) {materialFunction=&VectorLength}.V2 Length) {name=Core_Msk, nodeColor=<rgba 0.6875 1 0 1>}
  n148 = NamedRerouteUsage()
  n149 = VectorParameter() {
   group=Wrap_PlasmaCore_Core
   parameterName=PlasC_Core_Edge_Msk_Inner_Fallof
   channelNames {a="", b="", g=Max, r=Min}
   defaultValue=<rgba 0.4 0.5 0 0>
  }
  n150 = NamedRerouteUsage()
  n151 = SmoothStep(Min=@n149.Min, Max=@n149.Min, Value=@n150) {constMax=0.5, constMin=0.4}
  n152 = ScalarParameter() {defaultValue=0.5, group=Wrap_PlasmaCore_Plasma, parameterName=PlasC_Brightness}
  NamedRerouteDeclaration(None=Multiply(A=Multiply(A=@n151, B=@n148), B=@n152)) {name=Plasma_B_Colors, nodeColor=<rgba 0.4766 1 0 1>}
  n153 = VectorParameter() {
   group=Wrap_PlasmaCore_Core
   parameterName=PlasC_Core_Edge_Msk_Edge_Msk_Fallof
   channelNames {a="", b=Power, g=Max, r=Min}
   defaultValue=<rgba 1 0.5 2 0>
  }
  n154 = ScalarParameter() {defaultValue=127, group=Wrap_PlasmaCore_Plasma, parameterName="PlasC_CC_Edge  Index"}
  n155 = VectorParameter() {
   group=Wrap_PlasmaCore_Core
   parameterName=PlasC_Core_Edge_Msk_Edge_Blend_Fallof
   channelNames {a="", b="", g=Max, r=Min}
   defaultValue=<rgba 0.5 0.41 0 0>
  }
  n156 = Add(A=SmoothStep(Min=@n155.Min, Max=@n155.Min, Value=@n150) {constMax=0.41, constMin=0.5}, B=NamedRerouteUsage()) {constB=0}
  n157 = CurveAtlasRowParameter(CurveTime=@n156) {
   atlas=&CruveAtlas_Wraps_01
   curve=&CC_Plasma_C
   defaultValue=127
   group=Wrap_PlasmaCore_Plasma
   parameterName=PlasC_CC_Edge
  }
  n158 = Multiply(
   A=Power(Base=SmoothStep(Min=@n153.Min, Max=@n153.Min, Value=@n150) {constMax=0.5, constMin=1}, Exp=@n153.Power)
   B=Multiply(
    A=MaterialFunctionCall(Index=@n154, CCinput=@n157, UV=@n156) {materialFunction=&MF_Global_CC_Tex}.3V
    B=NamedRerouteUsage()
   )
  )
  NamedRerouteDeclaration(None=Multiply(A=Multiply(A=@n151, B=@n158), B=@n152)) {name=Plasma_B_Colors_Edge, nodeColor=<rgba 0.4766 1 0 1>}
  NamedRerouteDeclaration(None=MaterialFunctionCall!(UVs=@n3, Texture Size=ScalarParameter() {defaultValue=256, group=Wrap_PlasmaCore, parameterName=PlasC_Comput_mips}) {materialFunction=/Engine/Functions/Engine_MaterialFunctions01/Texturing/ComputeMipLevel}.MipLevel) {name=Mips, nodeColor=<rgba 0.3594 1 0 1>}
  n159 = VectorParameter() {group=Wrap_PlasmaCore_Back, parameterName=PlasC_Back_Color_A, defaultValue=<rgba 0.0057 0.0315 0.1406 1>}
  n160 = MaterialFunctionCall(Vector 2=Multiply(A=ScalarParameter() {defaultValue=1, group=Wrap_PlasmaCore_Back, parameterName=PlasC_Back_Size}, B=@n3)) {materialFunction=&VectorLength}
  n161 = ScalarParameter() {defaultValue=2, group=Wrap_PlasmaCore_Back, parameterName=PlasC_Back_Pow}
  NamedRerouteDeclaration(None=LinearInterpolate(
    A=VectorParameter() {group=Wrap_PlasmaCore_Back, parameterName=PlasC_Back_Color_B, defaultValue=<rgba 0.005 0.005 0.005 1>}.RGB
    B=@n159.RGB
    Alpha=Power(Base=@n160.V2 Length, Exp=@n161) {constExponent=3}
   )) {name=BackGround, nodeColor=<rgba 1 0.6797 0 1>}
  n162 = TextureObjectParameter() {
   group=Wrap_PlasmaCore
   parameterName=PlasC__Tex_MskB
   samplerType=SAMPLERTYPE_Masks
   texture=/Game/Weapons/WeaponSkins/Wraps/Textures/PlasmaCore/T_Wrap_PlasmaCore_MskB
  }
  NamedRerouteDeclaration(None=@n162) {name=Tex_MskB, nodeColor=<rgba 1 0.1406 0 1>}
  n163 = VectorParameter() {group=Wrap_PlasmaCore_Uvs, parameterName="PlasC_Rotation_Weapon\\PickAxe", defaultValue=<rgba 0 0 -0.25 1>}
  n164 = MaterialFunctionCall(Float4=NamedRerouteUsage()) {materialFunction=&BreakOutFloat4Components}
  n165 = LinearInterpolate(
   A=VectorParameter() {group=Wrap_PlasmaCore_Uvs, parameterName=PlasC_Rotation_Vehicle, defaultValue=<rgba 0 0 0 1>}.RGB
   B=VectorParameter() {group=Wrap_PlasmaCore_Uvs, parameterName=PlasC_Rotation_Vehicle_Side, defaultValue=<rgba 0 0 0 1>}.RGB
   Alpha=@n164.G
  )
  n166 = VectorParameter() {group=Wrap_PlasmaCore_Uvs, parameterName="PlasC_Rotation_Mech\\Character\\Backpack", defaultValue=<rgba 0.25 0.25 0 1>}
  NamedRerouteDeclaration(None=LinearInterpolate(A=LinearInterpolate(A=@n163.RGB, B=@n165, Alpha=@n164.R), B=@n166.RGB, Alpha=@n164.B)) {name=TS_Rotation, nodeColor=<rgba 1 0 0.375 1>}
  n167 = VectorParameter() {group=Wrap_PlasmaCore, parameterName=PlasC_Alt_Axis_Color, defaultValue=<rgba 0.26 0 0 0>}
  n168 = MaterialFunctionCall!(
   Texture=NamedRerouteUsage()
   Caculated_Bounds=NamedRerouteUsage()
   Pre_Skinned_Normals=NamedRerouteUsage()
   Blend_Contrast=Constant() {r=0.5}
   IsStatsB=NamedRerouteUsage()
   IsState=NamedRerouteUsage()
   Fill_alternate_Axis_Opacity=Constant() {r=1}
   "Fill_alternate_Axis_Color_A=Roughness"=AppendVector(A=@n167.RGB, B=@n167.RGB)
   Offset
   Scale=NamedRerouteUsage()
   Offset_Pivot
   Rotation_Axis=NamedRerouteUsage()
  ) {materialFunction=&/MF_TopsSideProjection_Base_V2}
  NamedRerouteDeclaration(None=ChannelMaskParameter(None=@n168.Combined) {group=Wrap_PlasmaCore, parameterName=PlasC_Roughness_Channel}.R) {name=Roughness, nodeColor=<rgba 1 0 0.2109 1>}
  NamedRerouteDeclaration(None=@n34) {name=reactive, nodeColor=<rgba 0 0.1562 1 1>}
  NamedRerouteDeclaration(None=@n145) {name=Warp_Curvature, nodeColor=<rgba 1 0.1406 0 1>}
  n169 = ChannelMaskParameter(None=@n168.Combined) {group=Wrap_PlasmaCore, maskChannel=Alpha, parameterName=PlasC_Warp_Channel, defaultValue=<rgba 0 0 0 1>}
  NamedRerouteDeclaration(None=@n169.R) {name=Warp, nodeColor=<rgba 1 0 0.2109 1>}
  n170 = ScalarParameter() {defaultValue=2, group=Wrap_PlasmaCore_Uvs, parameterName=PlasC_TileRate_Vehicle, sortPriority=150}
  n171 = ScalarParameter() {defaultValue=2, group=Wrap_PlasmaCore_Uvs, parameterName=PlasC_TileRate_Vehicle_Side, sortPriority=150}
  n172 = MaterialFunctionCall(Float4=NamedRerouteUsage()) {materialFunction=&BreakOutFloat4Components}
  n173 = LinearInterpolate(
   A=LinearInterpolate(
    A=ScalarParameter() {defaultValue=2, group=Wrap_PlasmaCore_Uvs, parameterName=PlasC_TileRate_Weapon, sortPriority=150}
    B=LinearInterpolate(A=@n170, B=@n171, Alpha=@n172.G)
    Alpha=@n172.R
   )
   B=LinearInterpolate(
    A=ScalarParameter() {defaultValue=2, group=Wrap_PlasmaCore_Uvs, parameterName=PlasC_TileRate_Mech, sortPriority=150}
    B=ScalarParameter() {defaultValue=2, group=Wrap_PlasmaCore_Uvs, parameterName=PlasC_TileRate_Character, sortPriority=150}
    Alpha=@n172.A
   )
   Alpha=@n172.B
  )
  n174 = ScalarParameter() {defaultValue=2, group=Wrap_PlasmaCore_Uvs, parameterName=PlasC_TileRate_Backpack, sortPriority=150}
  n175 = MaterialFunctionCall(Float2=NamedRerouteUsage()) {materialFunction=&BreakOutFloat2Components}
  NamedRerouteDeclaration(None=LinearInterpolate(
    A=LinearInterpolate(A=@n173, B=@n174, Alpha=@n175.R)
    B=ScalarParameter() {defaultValue=2, group=Wrap_PlasmaCore_Uvs, parameterName=PlasC_TileRate_Pickaxe, sortPriority=150}
    Alpha=@n175.G
   )) {name=UVs_Scale_V1, nodeColor=<rgba 0 0.7188 1 1>}
  NamedRerouteDeclaration(None=LinearInterpolate(
    A=ScalarParameter() {defaultValue=1, group=Wrap_PlasmaCore_Reactive, parameterName=PlasC_Reactive_Speed_Start}
    B=ScalarParameter() {group=Wrap_PlasmaCore_Reactive, parameterName=PlasC_Reactive_Speed_End}
    Alpha=@n34
   )) {name=Reactive_Speed, nodeColor=<rgba 0.1016 1 0 1>}
  NamedRerouteDeclaration(None=LinearInterpolate(
    A=ScalarParameter() {defaultValue=1, group=Wrap_PlasmaCore_Reactive, parameterName=PlasC_Reactive_Opact_Start}
    B=ScalarParameter() {group=Wrap_PlasmaCore_Reactive, parameterName=PlasC_Reactive_Opact_End}
    Alpha=@n34
   )) {name=Reactive_Opactiy, nodeColor=<rgba 0.1016 1 0 1>}
  NamedRerouteDeclaration(None=FunctionInput(Preview=Constant4Vector().RGBA) {bUsePreviewValueAsDefault=true, inputName=IsState, inputType=FunctionInput_Vector4, sortPriority=4}) {name=IsState, nodeColor=<rgba 0.1497 1 0.1633 1>}
  NamedRerouteDeclaration(None=LinearInterpolate(
    A=ScalarParameter() {defaultValue=3, group=Wrap_PlasmaCore_Reactive, parameterName=PlasC_Reactive_Bright_Start}
    B=ScalarParameter() {defaultValue=1, group=Wrap_PlasmaCore_Reactive, parameterName=PlasC_Reactive_Bright_End}
    Alpha=@n34
   )) {name=Reactive_Brightness, nodeColor=<rgba 0.1016 1 0 1>}
  n176 = CollectionParameter() {collection=&FortniteMaterialParameters, parameterName=EventAudio_Amplitude}
  n177 = CollectionParameter() {collection=&FortniteMaterialParameters, parameterName=Emote_AudioLow}
  n178 = MaterialFunctionCall(
   Input=Max(A=@n176, B=@n177)
   Input Low=ScalarParameter() {defaultValue=0.75, group=Wrap_PlasmaCore_Audio, parameterName=PlasC_Sectrum_Low_Remap, sortPriority=150}
   Input High=ScalarParameter() {defaultValue=0.9, group=Wrap_PlasmaCore_Audio, parameterName=PlasC_Sectrum_High_Remap, sortPriority=150}
   Target Low=Constant()
   Target High=Constant() {r=1}
  ) {materialFunction=&RemapValueRange}
  n179 = CollectionParameter() {collection=&FortniteMaterialParameters, parameterName=Emote_AudioHigh}
  NamedRerouteDeclaration(None=Max(A=Saturate(None=@n178.Result), B=@n179)) {name=AudioSpectrum, nodeColor=<rgba 1 0.8203 0 1>}
  NamedRerouteDeclaration(None=@n46.Half Extents) {name=Bounds_Green, nodeColor=<rgba 0 0.7656 1 1>}
  n180 = ScalarParameter() {defaultValue=0.5, group=Wrap_PlasmaCore_Plasma, parameterName=PlasC_Plasma_Msk_outer_Size}
  NamedRerouteDeclaration(None=SmoothStep(
    Min=ScalarParameter() {defaultValue=1, group=Wrap_PlasmaCore_Plasma, parameterName=PlasC_Plasma_Msk_outer_min}
    Max=ScalarParameter() {defaultValue=0.99, group=Wrap_PlasmaCore_Plasma, parameterName=PlasC_Plasma_Msk_outer_Max}
    Value=Multiply(A=@n66.V2 Length, B=@n180)
   ) {constMax=0, constMin=1}) {name=Msk_Plasma_outer_Msk, nodeColor=<rgba 1 0.5859 0 1>}
  n181 = SmoothStep(
   Min=ScalarParameter() {defaultValue=0.99, group=Wrap_PlasmaCore_Warp, parameterName=PlasC_Plasma_Warp_Falloff_Hardness}
   Value=Multiply(A=@n66.V2 Length, B=ScalarParameter() {defaultValue=0.25, group=Wrap_PlasmaCore_Warp, parameterName=PlasC_Plasma_Warp_Falloff_Size})
  )
  NamedRerouteDeclaration(None=@n181) {name=Msk_Plasma_Warp_Falloff, nodeColor=<rgba 1 0.5859 0 1>}
  NamedRerouteDeclaration(None=@n58.EmissiveColor) {name=Root_Emissive, nodeColor=<rgba 1 0.2109 0 1>}
  NamedRerouteDeclaration(None=@n58.BaseColor) {name=Root_BC, nodeColor=<rgba 0.4062 0 1 1>}
  NamedRerouteDeclaration(None=FunctionInput(Preview=Constant2Vector()) {bUsePreviewValueAsDefault=true, inputName=IsState_B, inputType=FunctionInput_Vector2, sortPriority=4}) {name=IsState_B, nodeColor=<rgba 0.1497 1 0.1633 1>}
  n182 = PreSkinnedLocalBounds()
  n183 = Subtract(A=VertexInterpolator(VS=PreSkinnedPosition()).PS, B=@n182.Min)
  n184 = MaterialFunctionCall(Color=Subtract(A=@n182.Max, B=@n182.Max)) {materialFunction=&SplitComponents}
  n185 = Max(A=Max(A=@n184.R, B=@n184.R), B=@n184.B)
  NamedRerouteDeclaration(None=FunctionInput(Preview=Divide(A=@n183, B=@n185)) {bUsePreviewValueAsDefault=true, inputName=Calcated_Bounds, sortPriority=11}) {name=Caclcated_Bounds, nodeColor=<rgba 0 0.9297 1 1>}
  n186 = TextureObjectParameter() {
   group=Wrap_PlasmaCore
   parameterName=PlasC__Tex_Msk
   samplerType=SAMPLERTYPE_Masks
   texture=/Game/Weapons/WeaponSkins/Wraps/Textures/PlasmaCore/T_Wrap_PlasmaCore_Msk
  }
  NamedRerouteDeclaration(None=@n186) {name=Tex_Msk, nodeColor=<rgba 1 0.1406 0 1>}
  n187 = PreSkinnedNormal()
  NamedRerouteDeclaration(None=FunctionInput(Preview=VertexInterpolator(VS=@n187).PS) {bUsePreviewValueAsDefault=true, inputName=Pre-Skinned Local Normal, sortPriority=10}) {name=PreSkin_Normal, nodeColor=<rgba 0.8281 1 0 1>}
  NamedRerouteDeclaration(None=@n58.Normal) {name=Root_Normal, nodeColor=<rgba 0.4062 0 1 1>}
  NamedRerouteUsage()
  NamedRerouteUsage()
  NamedRerouteUsage()
  Saturate(None=@n99)
  Saturate(None=@n35)
  Saturate(None=@n181)
  ScalarParameter() {defaultValue=-9999999, group=TopSideProjection_Emissive_Reactive, parameterName=Time_BP}
 ]}
