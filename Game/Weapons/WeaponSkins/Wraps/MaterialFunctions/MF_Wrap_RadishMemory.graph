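# Weapon wrap material function for RadishMemory: an animated galaxy-and-cloud skin, enabled by the Wrap_RadishMemory static switch.
# Rotating camera-vector triplanar samples build a tinted starfield and edge mask.
# Two scales of scrolling, distorted noise, Fresnel and AO/curvature masks drive curve-atlas-colored clouds for the emissive.
# The Primary mask blends this emissive over a black base; the Secondary mask blends in a tinted metallic layer with view-dependent glow.
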MaterialFunction /Game/Weapons/WeaponSkins/Wraps/MaterialFunctions/MF_Wrap_RadishMemory
graph {nodes [
  n1 = Multiply(A=Time(), B=VectorParameter() {group=RadishMemory, parameterName=RM_Small Star_Pan_Speed, defaultValue=<rgba 0 0 0 1>}.RGB)
  Add(
   A=@n1
   B=Multiply(A=NamedRerouteUsage(), B=ScalarParameter() {defaultValue=0.05, group=RadishMemory, parameterName=RM_GalaxyWarpAmount}) {constB=0.2}
  )
  n2 = WorldPosition()
  n3 = Multiply(A=Transform(None=VectorParameter() {group=RadishMemory_RimLight, parameterName=Rim_Vector, defaultValue=<rgba 0 0 1 1>}.RGB) {transformSourceType=TRANSFORMSOURCE_View}) {constB=-1}
  n4 = Constant() {r=-1}
  n5 = DotProduct(A=@n2.XYZ, B=Add(A=@n3, B=@n4))
  n6 = ComponentMask(None=Multiply(A=@n4) {constB=-1}) {g=true, r=true}
  n7 = Constant()
  n8 = DotProduct(A=@n2.XYZ, B=Add(A=@n3, B=AppendVector(A=@n6, B=@n7)))
  n9 = ScalarParameter() {defaultValue=2, group=RadishMemory_RimLight, parameterName=Highlight_RimContrast, sortPriority=10}
  n10 = Power(Base=OneMinus(None=Add(A=@n5, B=@n8)), Exp=@n9) {constExponent=3}
  n11 = Constant() {r=1}
  n12 = DotProduct(A=@n2.XYZ, B=Add(A=@n3, B=@n11))
  n13 = ComponentMask(None=Multiply(A=@n11) {constB=-1}) {g=true, r=true}
  n14 = Constant()
  n15 = DotProduct(A=@n2.XYZ, B=Add(A=@n3, B=AppendVector(A=@n13, B=@n14)))
  n16 = Power(Base=OneMinus(None=Add(A=@n12, B=@n15)), Exp=@n9)
  n17 = Fresnel(ExponentIn=ScalarParameter() {defaultValue=2, group=RadishMemory_RimLight, parameterName=Highlight_FresnelAmount, sortPriority=10}) {exponent=2}
  n18 = Power(Base=@n17) {constExponent=8}
  n19 = Power(Base=Saturate(None=Add(A=@n17, B=@n18))) {constExponent=6}
  Add(A=Max(A=Saturate(None=@n10), B=Saturate(None=@n16)), B=@n19)
  n20 = Multiply(A=ComponentMask(None=NamedRerouteUsage()) {b=true}, B=NamedRerouteUsage())
  n21 = Constant3Vector() {constant=<rgba 0.3273 0.2501 0.5528 1>}
  n22 = Multiply(
   A=MaterialFunctionCall!(In=NamedRerouteUsage(), Contrast=Constant() {r=0.15}) {materialFunction=/Engine/Functions/Engine_MaterialFunctions01/ImageAdjustment/CheapContrast_RGB}.Result
   B=@n21
  )
  n23 = Constant3Vector() {constant=<rgba 0.1232 0 0.1981 1>}
  n24 = Saturate(None=Multiply(A=NamedRerouteUsage()) {constB=2})
  n25 = ScalarParameter() {defaultValue=119, group=RadishMemory_Textures, parameterName="RM_Noise_ColorCurve  Index", sortPriority=399}
  n26 = Saturate(None=NamedRerouteUsage())
  n27 = CurveAtlasRowParameter(CurveTime=@n26) {
   atlas=/Game/Weapons/WeaponSkins/Wraps/Textures/CruveAtlas_Wraps_01
   curve=/Game/Weapons/WeaponSkins/Wraps/Textures/RadishMem/C_RadishMemory
   defaultValue=119
   group=RadishMemory_Textures
   parameterName=RM_Noise_ColorCurve
   sortPriority=399
  }
  n28 = TextureSample() {samplerSource=SSM_Wrap_WorldGroupSettings, samplerType=SAMPLERTYPE_Masks, texture=&T_BlackMask}
  n29 = FunctionInput(Preview=@n28.A) {bUsePreviewValueAsDefault=true, inputName=Mask GMAP, inputType=FunctionInput_Scalar, sortPriority=24, previewValue=<1 0 0 1>}
  n30 = Power(Base=OneMinus(None=@n29)) {constExponent=0.5}
  n31 = FunctionInput(Preview=@n28.G) {bUsePreviewValueAsDefault=true, inputName=Mask Crevice, inputType=FunctionInput_Scalar, sortPriority=24, previewValue=<1 0 0 1>}
  n32 = LinearInterpolate(Alpha=@n31) {constA=-1}
  n33 = Multiply(A=Power(Base=@n32) {constExponent=1.5}, B=ScalarParameter() {defaultValue=3, group=RadishMemory, parameterName=RM_CurvatureIntensity01}) {constB=3}
  n34 = Multiply(A=Add(A=NamedRerouteUsage(), B=NamedRerouteUsage()), B=NamedRerouteUsage())
  n35 = Multiply(
   A=Add(
    A=NamedRerouteDeclaration(None=@n30) {name=AO, nodeColor=<rgba 0.1953 0 1 1>}
    B=NamedRerouteDeclaration(None=@n33) {name=Curvature, nodeColor=<rgba 1 0 0.3047 1>}
   )
   B=@n34
  )
  n36 = Multiply(A=@n35, B=NamedRerouteUsage())
  n37 = Saturate(None=NamedRerouteUsage())
  n38 = CurveAtlasRowParameter(CurveTime=@n37) {
   atlas=&CruveAtlas_Wraps_01
   curve=&C_RadishMemory
   defaultValue=119
   group=RadishMemory_Textures
   parameterName=RM_Noise_ColorCurve
   sortPriority=399
  }
  n39 = Multiply(
   A=Multiply(A=Power(Base=@n32) {constExponent=4}, B=ScalarParameter() {defaultValue=20, group=RadishMemory, parameterName=RM_CurvatureIntensity02}) {constB=20}
   B=@n34
  )
  n40 = Multiply(A=@n39, B=NamedRerouteUsage())
  n41 = Power(Base=Multiply(A=NamedRerouteUsage(), B=NamedRerouteUsage())) {constExponent=1}
  n42 = LinearInterpolate(
   A=LinearInterpolate(
    A=LinearInterpolate(A=@n22, B=@n23, Alpha=@n24)
    B=MaterialFunctionCall!(Index=@n25, CCinput=@n27, CCinput_Alpha, UV=@n26) {materialFunction=&/MF_Global_CC_Tex}.3V
    Alpha=Saturate(None=@n36)
   )
   B=MaterialFunctionCall(Index=@n25, CCinput=@n38, UV=@n37) {materialFunction=&MF_Global_CC_Tex}.3V
   Alpha=Saturate(None=Add(A=@n40, B=@n41))
  )
  n43 = Constant3Vector() {constant=<rgba 0 0.2241 0.4249 1>}
  n44 = Multiply(
   A=Multiply(A=NamedRerouteUsage(), B=NamedRerouteUsage())
   B=ScalarParameter() {defaultValue=1, group=RadishMemory, parameterName=RM_CloudEdgeIntensity}
  )
  n45 = LinearInterpolate(A=Add(A=Clamp(None=@n20) {minDefault=-0.1}, B=LinearInterpolate(A=@n42, B=@n43, Alpha=@n44)), B=@n22, Alpha=@n24)
  n46 = NamedRerouteUsage()
  Add(A=@n45, B=@n46)
  n47 = FunctionInput(Preview=@n28.RGBA) {inputName=In CM Mask, inputType=FunctionInput_Vector4}
  ComponentMask(None=@n47) {g=true}
  ComponentMask(None=@n47) {a=true}
  n48 = TextureSample() {samplerType=SAMPLERTYPE_Masks, texture=/Game/Weapons/FORT_Rifles/Textures/SCAR_S}
  n49 = SetMaterialAttributes!(
   MaterialAttributes
   BaseColor=TextureSample() {texture=/Game/Weapons/FORT_Rifles/Textures/SCAR_D}.RGB
   Metallic=@n48.G
   Normal=TextureSample() {samplerType=SAMPLERTYPE_Normal, texture=/Game/Weapons/FORT_Rifles/Textures/SCAR_N}.RGB
   Roughness=@n48.G
   Specular=@n48.G
  ) {attributeSetTypes [
    #1
    #2
    #3
    #4
    #5
   ]}
  n50 = FunctionInput(Preview=@n49) {inputName=Material, inputType=FunctionInput_MaterialAttributes, sortPriority=-10}
  n51 = Constant()
  n52 = SetMaterialAttributes!(
   MaterialAttributes
   BaseColor=@n51
   Metallic=@n51
   Specular=@n51
   Roughness=Constant() {r=1}
   EmissiveColor=NamedRerouteUsage()
   Normal=NamedRerouteDeclaration(None=GetMaterialAttributes(None=@n50) {attributeGetTypes [#3]}.Normal) {name=InNormal, nodeColor=<rgba 1 0.9844 0 1>}
  ) {attributeSetTypes [
    #1
    #2
    #5
    #4
    #6
    #3
   ]}
  n53 = FunctionInput() {inputName=Primary Mask, inputType=FunctionInput_Scalar, sortPriority=-9}
  n54 = BlendMaterialAttributes(A=@n50, B=@n52, Alpha=@n53)
  n55 = VectorParameter() {group=RadishMemory, parameterName=RM_SecondaryBaseColor, sortPriority=-10, defaultValue=<rgba 0.0273 0.0091 0.107 1>}
  n56 = VectorParameter() {group=RadishMemory, parameterName=RM_SecondaryEmissive, sortPriority=-10, defaultValue=<rgba 0 0.0482 0.0595 1>}
  n57 = DotProduct(A=CameraVectorWS(), B=VertexNormalWS())
  n58 = Saturate(None=Multiply(A=Power(Base=@n57) {constExponent=8}) {constB=0.075})
  n59 = Saturate(None=Power(Base=@n57, Exp=ScalarParameter() {defaultValue=4, group=RadishMemory, parameterName=RM_SecondaryFresnel, sortPriority=-10}))
  n60 = SetMaterialAttributes!(
   MaterialAttributes=@n54
   BaseColor=@n55.RGB
   Metallic=ScalarParameter() {defaultValue=1, group=RadishMemory, parameterName=RM_SecondaryMetallic, sortPriority=-10}
   Specular=ScalarParameter() {defaultValue=0.6, group=RadishMemory, parameterName=RM_SecondarySpecular, sortPriority=-10}
   Roughness=ScalarParameter() {defaultValue=0.2, group=RadishMemory, parameterName=RM_SecondaryRoughness, sortPriority=-10}
   EmissiveColor=Multiply(A=@n56.RGB, B=Add(A=@n58, B=@n59))
  ) {attributeSetTypes [
    #1
    #2
    #5
    #4
    #6
   ]}
  n61 = FunctionInput() {inputName=Secondary Mask_1, inputType=FunctionInput_Scalar, sortPriority=-8}
  FunctionOutput(None=StaticSwitchParameter(True=BlendMaterialAttributes(A=@n54, B=@n60, Alpha=@n61), False=@n50) {group=Customization_Wrap, parameterName=Wrap_RadishMemory})
  n62 = ScalarParameter() {group=RadishMemory, parameterName=RM_CloudOpacity}
  n63 = ScalarParameter() {defaultValue=1, group=RadishMemory, parameterName=RM_CloudOpacitysides}
  n64 = NamedRerouteUsage()
  n65 = Fresnel(
   ExponentIn=LinearInterpolate(
    A=ScalarParameter() {defaultValue=0.75, group=RadishMemory, parameterName=RM_CloudOpacityFresnelLargeLo}
    B=ScalarParameter() {defaultValue=8, group=RadishMemory, parameterName=RM_CloudOpacityFresnelLargeHi}
    Alpha=NamedRerouteUsage()
   )
   Normal=@n64
  ) {baseReflectFraction=0}
  n66 = Fresnel(
   ExponentIn=LinearInterpolate(
    A=ScalarParameter() {defaultValue=0.5, group=RadishMemory, parameterName=RM_CloudOpacityFresnelSmallLo}
    B=ScalarParameter() {defaultValue=7, group=RadishMemory, parameterName=RM_CloudOpacityFresnelSmallHi}
    Alpha=NamedRerouteUsage()
   )
   Normal=@n64
  ) {baseReflectFraction=0}
  n67 = Saturate(None=Add(A=LinearInterpolate(A=@n62, B=@n63, Alpha=@n65), B=LinearInterpolate(A=@n62, B=@n63, Alpha=@n66)))
  LinearInterpolate(A=Power(Base=@n67) {constExponent=1.75}, B=@n34) {constAlpha=0.25}
  n68 = NamedRerouteUsage()
  n69 = NamedRerouteUsage()
  n70 = LinearInterpolate(
   A=ScalarParameter() {defaultValue=0.2, group=RadishMemory, parameterName=RM_CloudOpacityFresnelGeneralLo}
   B=ScalarParameter() {defaultValue=2, group=RadishMemory, parameterName=RM_CloudOpacityFresnelGeneralHi}
   Alpha=Saturate(None=Add(A=@n68, B=@n69))
  )
  n71 = NamedRerouteUsage()
  n72 = Multiply(
   A=ScalarParameter() {defaultValue=1, group=RadishMemory, parameterName=RM_CloudEdgeIntensity}
   B=Power(Base=Fresnel(ExponentIn=@n70, Normal=@n71) {baseReflectFraction=0}) {constExponent=5}
  )
  MaterialFunctionCall!(Blend=@n67, Base=NamedRerouteDeclaration(None=Saturate(None=@n72)) {name=GeneralFresnel, nodeColor=<rgba 0 1 0.0391 1>}) {materialFunction=/Engine/Functions/Engine_MaterialFunctions03/Blends/Blend_Screen}
  n73 = TextureObject() {samplerType=SAMPLERTYPE_Masks, texture=/Game/Global/Textures/Default/Noise/T_WrapEdgeWear_Noise}
  n74 = ScalarParameter() {defaultValue=0.33, group=RadishMemory, parameterName=BG Color GalaxyTiling}
  n75 = ScalarParameter() {defaultValue=0.005, group=RadishMemory, parameterName=BG Color Rotation Speed}
  n76 = ScalarParameter() {defaultValue=-0.005, group=RadishMemory, parameterName=BG Color Rotation Speed_1}
  n77 = Multiply(
   A=Time()
   B=AppendVector(A=AppendVector(A=@n75, B=@n76), B=ScalarParameter() {defaultValue=0.033, group=RadishMemory, parameterName="BG Color  Rotation Speed_2"})
  )
  MaterialFunctionCall!(
   CameraVector
   Rotation Speed=@n77
   Texture=@n73
   Tiling=@n74
   Offset
   AxisFadeContrast=ScalarParameter() {defaultValue=6, group=RadishMemory, parameterName=BG Color AxisFade}
  ) {materialFunction=/Game/Weapons/WeaponSkins/MaterialFunction/TriplanarCameraVector_Rotated}
  Multiply(A=@n46) {constB=0.2}
  n78 = Multiply(A=TextureCoordinate(), B=@n74)
  n79 = ComponentMask(None=@n77) {g=true, r=true}
  n80 = Power(
   Base=TextureSample(UVs=Add(A=@n78, B=@n79), Tex=@n73).RGB
   Exp=ScalarParameter() {defaultValue=2.5, group=RadishMemory, parameterName=BG Color Contrast}
  ) {constExponent=2.5}
  n81 = ComponentMask(None=@n80) {r=true}
  Multiply(A=@n81, B=ScalarParameter() {defaultValue=1.1, group=RadishMemory, parameterName=BG_Color01Amount, sliderMax=1.5})
  n82 = ComponentMask(None=@n80) {g=true}
  Multiply(A=@n82, B=ScalarParameter() {defaultValue=1.2, group=RadishMemory, parameterName=BG_Color03Amount, sliderMax=1.5})
  n83 = ComponentMask(None=@n80) {b=true}
  Multiply(A=@n83, B=ScalarParameter() {defaultValue=1.3, group=RadishMemory, parameterName=BG_Color02Amount, sliderMax=1.5})
  NamedRerouteDeclaration(None=@n21) {name=LightPurple, nodeColor=<rgba 1 0.4922 0 1>}
  n84 = Multiply(A=Time(), B=ScalarParameter() {defaultValue=-0.0033, group=RadishMemory, parameterName=RM_CloudLarge_XSpeed})
  n85 = Multiply(A=TextureCoordinate(), B=ScalarParameter() {defaultValue=0.5, group=RadishMemory, parameterName=RM_CloudLarge_XTileAmount})
  n86 = NamedRerouteUsage()
  n87 = ScalarParameter() {defaultValue=0.2, group=RadishMemory, parameterName=RM_CloudLarge_Distortion}
  n88 = TextureObjectParameter() {group=RadishMemory_Textures, parameterName=RM_Cloud_Noise, sortPriority=100, texture=&T_BlacksRGB}
  n89 = TextureSample(UVs=Add(A=Add(A=@n84, B=@n85), B=Multiply(A=@n86, B=@n87) {constB=0.1}), Tex=@n88) {samplerSource=SSM_Wrap_WorldGroupSettings, texture=&T_BlacksRGB}
  NamedRerouteDeclaration(None=@n89.R) {name=LargeNoiseA, nodeColor=<rgba 0.8047 0 1 1>}
  n90 = TextureCoordinate() {uTiling=0.5, vTiling=0.5}
  n91 = Time()
  n92 = Multiply(A=@n91, B=ScalarParameter() {defaultValue=0.005, group=RadishMemory, parameterName=RM_HeatWave_LargeSpeed})
  n93 = TextureObject() {texture=&T_BlacksRGB}
  NamedRerouteDeclaration(None=TextureSample(UVs=Add(A=@n90, B=@n92), Tex=@n93) {samplerSource=SSM_Wrap_WorldGroupSettings, texture=&2ndInfectedMapLinear}.R) {name=HeatWaveLarge, nodeColor=<rgba 0.7344 0 1 1>}
  n94 = Power(Base=@n89.R) {constExponent=1.25}
  n95 = ScalarParameter() {defaultValue=15, group=RadishMemory, parameterName=RM_CloudLarge_Intensity}
  NamedRerouteDeclaration(None=Saturate(None=Multiply(A=@n94, B=@n95) {constB=15})) {name=LargeNoise, nodeColor=<rgba 0 1 0.0859 1>}
  n96 = PreSkinnedNormal()
  NamedRerouteDeclaration(None=ShaderStageSwitch(PixelShader=VertexInterpolator(VS=@n96).PS, VertexShader=@n96)) {name=PreskinnedNormal, nodeColor=<rgba 0 0.7422 1 1>}
  n97 = Time()
  n98 = Multiply(A=@n97, B=Constant() {r=0.02})
  n99 = ScalarParameter() {group=RadishMemory, parameterName=RM_EdgeMask_Rotation}
  n100 = AppendVector(
   A=ScalarParameter() {defaultValue=0.033, group=RadishMemory, parameterName=RM_EdgeMask_Rotation Speed}
   B=ScalarParameter() {defaultValue=-0.033, group=RadishMemory, parameterName=RM_EdgeMask_Rotation Speed_1}
  )
  n101 = ScalarParameter() {defaultValue=0.05, group=RadishMemory, parameterName=RM_EdgeMask_Rotation Speed_2}
  n102 = MaterialFunctionCall(
   Rotation Speed=Multiply(A=LinearInterpolate(A=@n97, B=@n98, Alpha=@n99), B=AppendVector(A=@n100, B=@n101))
   Texture=TextureObjectParameter() {group=RadishMemory_Textures, parameterName=RM_EdgeNoise_Texture, texture=&T_BlacksRGB}
   Tiling=ScalarParameter() {defaultValue=2, group=RadishMemory, parameterName=RM_EdgeMask_GalaxyTiling}
   AxisFadeContrast=ScalarParameter() {defaultValue=6, group=RadishMemory, parameterName=RM_EdgeMask_AxisFade}
  ) {materialFunction=&TriplanarCameraVector_Rotated}
  NamedRerouteDeclaration(None=MaterialFunctionCall(In=Add(A=@n102.Result) {constB=0.4}, Contrast=ScalarParameter() {defaultValue=1.8, group=RadishMemory, parameterName=RM_EdgeMaskContrast}) {materialFunction=&CheapContrast}.Result) {name=EdgeMask, nodeColor=<rgba 0.4531 0 1 1>}
  n103 = Fresnel(
   ExponentIn=ScalarParameter() {defaultValue=2, group=RadishMemory_RimLight, parameterName=RM_CloudFresnelAmount}
   BaseReflectFractionIn=ScalarParameter() {group=RadishMemory_RimLight, parameterName=RM_CloudFresnelBaseReflect}
   Normal=NamedRerouteUsage()
  ) {exponent=2}
  NamedRerouteDeclaration(None=Saturate(None=Power(Base=@n103) {constExponent=6})) {name=FresnelBase, nodeColor=<rgba 1 0 0.5625 1>}
  NamedRerouteDeclaration(None=Transform(None=NamedRerouteUsage())) {name=NormalFresnel, nodeColor=<rgba 0.1719 1 0 1>}
  n104 = Add(A=Add(A=@n35, B=@n39), B=@n34)
  n105 = Multiply(A=ScalarParameter() {defaultValue=0.6, group=RadishMemory, parameterName=RM_EdgeMaskIntensity}, B=@n104)
  NamedRerouteDeclaration(None=LinearInterpolate(A=NamedRerouteUsage(), B=@n45, Alpha=Saturate(None=@n105))) {name=FinalEmissive, nodeColor=<rgba 1 0 0.4219 1>}
  n106 = ScalarParameter() {defaultValue=1, group=RadishMemory_RimLight, parameterName=Highlight_Brightness, sortPriority=10}
  NamedRerouteDeclaration(None=Saturate(None=Multiply(A=@n106, B=@n19))) {name=FinalRim, nodeColor=<rgba 1 0.4453 0 1>}
  n107 = AppendVector(
   A=ScalarParameter() {defaultValue=0.033, group=RadishMemory, parameterName=RM_Galaxy Rotation Speed}
   B=ScalarParameter() {defaultValue=-0.03, group=RadishMemory, parameterName=RM_Galaxy Rotation Speed_1}
  )
  n108 = ScalarParameter() {defaultValue=0.05, group=RadishMemory, parameterName=RM_Galaxy Rotation Speed_2}
  n109 = MaterialFunctionCall(
   Rotation Speed=Multiply(A=Time(), B=AppendVector(A=@n107, B=@n108))
   Texture=TextureObjectParameter() {group=RadishMemory_Textures, parameterName=RM_GalaxyTexture, texture=&T_BlacksRGB}
   Tiling=ScalarParameter() {defaultValue=0.5, group=RadishMemory, parameterName=RM_GalaxyTiling}
   AxisFadeContrast=ScalarParameter() {defaultValue=6, group=RadishMemory, parameterName=RM_GalaxyAxisFade}
  ) {materialFunction=&TriplanarCameraVector_Rotated}
  n110 = ScalarParameter() {defaultValue=1, group=RadishMemory, parameterName=RM_GalaxyBrightnessAmount}
  n111 = Add(
   A=Multiply(
    A=Multiply(A=@n109.Result, B=@n110)
    B=VectorParameter() {group=RadishMemory, parameterName=RM_BGMultTint, defaultValue=<rgba 0.1095 0.2307 0.5271 1>}.RGB
   )
   B=VectorParameter() {group=RadishMemory, parameterName=RM_BGAddTint, defaultValue=<rgba 0.0112 0.0138 0.02 1>}.RGB
  )
  n112 = Desaturation(None=@n111, Fraction=ScalarParameter() {defaultValue=0.5, group=RadishMemory, parameterName=RM_Star Saturation})
  n113 = Multiply(A=@n112, B=VectorParameter() {group=RadishMemory, parameterName=RM_Star Brightness, defaultValue=<rgba 50 50 50 1>}.RGB)
  n114 = Time()
  n115 = ScalarParameter() {defaultValue=-0.0125, group=RadishMemory, parameterName=RM_Small Star Rotation speed}
  n116 = TextureObjectParameter() {
   group=RadishMemory_Textures
   parameterName=RM_StarMask
   samplerType=SAMPLERTYPE_Masks
   texture=&T_BlackMask
  }
  n117 = ComponentMask(None=MaterialFunctionCall(
    Rotation Speed=Multiply(A=@n114, B=@n115)
    Texture=@n116
    Tiling=ScalarParameter() {defaultValue=8, group=RadishMemory, parameterName=RM_Small Star Tiling}
    Offset=@n1
   ) {materialFunction=&TriplanarCameraVector_Rotated}.Result) {b=true}
  n118 = Multiply(A=@n117, B=@n117)
  n119 = VectorParameter() {group=RadishMemory, parameterName=RM_Star Brightness_1, defaultValue=<rgba 200 200 200 1>}
  n120 = ComponentMask(None=MaterialFunctionCall!(
    CameraVector
    Texture=@n116
    Tiling=ScalarParameter() {defaultValue=3, group=RadishMemory, parameterName=RM_Bright Star Tiling}
    Offset
    AxisFadeContrast
   ) {materialFunction=/Engine/Functions/Engine_MaterialFunctions01/Texturing/TriplanarCameraVector}.Result) {r=true}
  n121 = Multiply(A=@n120, B=@n120)
  n122 = TextureSample(UVs=Panner(Coordinate=TextureCoordinate() {uTiling=2, vTiling=2}) {speedX=0.06, speedY=0.1}, Tex=@n116) {
   samplerSource=SSM_Wrap_WorldGroupSettings
   samplerType=SAMPLERTYPE_Masks
   texture=/Game/Characters/Player/Male/Medium/Bodies/M_Med_Soldier_04/Skins/Galaxy/Textures/Celestial_T_Stars
  }
  NamedRerouteDeclaration(None=LinearInterpolate(A=LinearInterpolate(A=@n111, B=@n113, Alpha=@n118), B=Multiply(A=@n112, B=@n119.RGB), Alpha=Multiply(A=@n121, B=@n122.G))) {name=Galaxy, nodeColor=<rgba 0 0.25 1 1>}
  n123 = Multiply(A=@n91, B=ScalarParameter() {defaultValue=0.0125, group=RadishMemory, parameterName=RM_HeatWave_SmallSpeed})
  NamedRerouteDeclaration(None=TextureSample(UVs=Add(A=@n90, B=@n123), Tex=@n93) {samplerSource=SSM_Wrap_WorldGroupSettings, texture=&2ndInfectedMapLinear}.G) {name=HeatwaveSmall, nodeColor=<rgba 1 0.8906 0 1>}
  NamedRerouteDeclaration(None=@n103) {name=FresnelMobile, nodeColor=<rgba 1 0 0.2109 1>}
  NamedRerouteDeclaration(None=@n65) {name=LargeFresnel, nodeColor=<rgba 1 0 0.2109 1>}
  NamedRerouteDeclaration(None=@n66) {name=SmallFresnel, nodeColor=<rgba 0.9453 0 1 1>}
  n124 = LinearInterpolate(
   A=VectorParameter() {group=RadishMemory, parameterName=BG_Color04, defaultValue=<rgba 0.1736 0 0.75 1>}.RGB
   B=VectorParameter() {group=RadishMemory, parameterName=BG_Color01, defaultValue=<rgba 0.75 0 0.0274 1>}.RGB
   Alpha=Saturate(None=@n81)
  )
  n125 = VectorParameter() {group=RadishMemory, parameterName=BG_Color03, defaultValue=<rgba 0 0.75 0.5434 1>}
  n126 = Saturate(None=@n82)
  NamedRerouteDeclaration(None=LinearInterpolate(
    A=LinearInterpolate(A=@n124, B=@n125.RGB, Alpha=@n126)
    B=VectorParameter() {group=RadishMemory, parameterName=BG_Color02, defaultValue=<rgba 0.75 0.3371 0.0193 1>}.RGB
    Alpha=Saturate(None=@n83)
   )) {name=ColorRotate, nodeColor=<rgba 0 0.0859 1 1>}
  n127 = Multiply(A=Time(), B=ScalarParameter() {defaultValue=-0.01, group=RadishMemory, parameterName=RM_CloudSmall_XSpeed})
  n128 = Multiply(A=TextureCoordinate(), B=ScalarParameter() {defaultValue=2, group=RadishMemory, parameterName=RM_CloudSmall_XTileAmount})
  n129 = NamedRerouteUsage()
  n130 = ScalarParameter() {defaultValue=0.5, group=RadishMemory, parameterName=RM_CloudSmall_Distortion}
  n131 = TextureSample(UVs=Add(A=Add(A=@n127, B=@n128), B=Multiply(A=@n129, B=@n130) {constB=0.1}), Tex=@n88) {samplerSource=SSM_Wrap_WorldGroupSettings, texture=&T_BlacksRGB}
  n132 = Power(Base=@n131.R) {constExponent=1.33}
  n133 = ScalarParameter() {defaultValue=20, group=RadishMemory, parameterName=RM_CloudSmall_Intensity}
  NamedRerouteDeclaration(None=Saturate(None=Multiply(A=@n132, B=@n133) {constB=20})) {name=SmallNoise, nodeColor=<rgba 0.0547 1 0 1>}
  NamedRerouteDeclaration(None=@n131.R) {name=SmallNoiseA, nodeColor=<rgba 0.0781 0 1 1>}
  NamedRerouteUsage()
  Power(Base=@n104) {constExponent=1.5}
  Saturate(None=Add(A=@n53, B=@n61))
 ]}
