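# Material function for the Renegade Raider Spark wrap, layering a fiery base and an accent material over the incoming material by masks, behind a static switch.
# Projected masks, normals and scrolling noise use per-item-type UV settings for weapons, vehicles, mechs, characters, backpacks and pickaxes.
# Noise, curvature and a gradient drive color-curve fire emissive, with a separate curve for crack glow.
# A Blueprint timestamp parameter derives timed spark, crack and soot fades.
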MaterialFunction /Game/Weapons/WeaponSkins/Wraps/MaterialFunctions/MF_Wrap_RenRaiderSpark
graph {nodes [
  n1 = TextureSample() {samplerType=SAMPLERTYPE_Masks, texture=/Game/Weapons/FORT_Rifles/Textures/T_Scar_CM}
  n2 = FunctionInput(Preview=@n1.R) {bUsePreviewValueAsDefault=true, inputName=Mask Primary, inputType=FunctionInput_Scalar, sortPriority=24, previewValue=<1 0 0 1>}
  n3 = FunctionInput(Preview=@n1.B) {bUsePreviewValueAsDefault=true, inputName=Mask Secondary, inputType=FunctionInput_Scalar, sortPriority=24, previewValue=<1 0 0 1>}
  Add(A=@n2, B=@n3)
  n4 = Time()
  Add(A=@n4)
  n5 = Add(A=NamedRerouteUsage()) {constB=-0.5}
  n6 = Saturate(None=Multiply(A=@n5) {constB=2})
  n7 = LinearInterpolate(
   B=Constant3Vector() {constant=<rgba 0 0 -1 0>}
   Alpha=MaterialFunctionCall(Float4=NamedRerouteUsage()) {materialFunction=&BreakOutFloat4Components}.B
  )
  n8 = ScalarParameter() {defaultValue=2, group=Wrap_RenegadeRaiderSpark, parameterName=Ren_Tri_Contrast}
  n9 = VectorParameter() {group=Wrap_RenegadeRaiderSpark, parameterName=Ren_Msk_Alternate}
  n10 = NamedRerouteUsage()
  n11 = MaterialFunctionCall!(
   Texture=NamedRerouteUsage()
   Caculated_Bounds=Add(A=NamedRerouteUsage(), B=@n7)
   Pre_Skinned_Normals=NamedRerouteUsage()
   Blend_Contrast=@n8
   IsStatsB=NamedRerouteUsage()
   IsState=NamedRerouteUsage()
   Fill_alternate_Axis_Opacity=Constant() {r=1}
   "Fill_alternate_Axis_Color_A=Roughness"=AppendVector(A=@n9.RGB, B=@n9.RGB)
   Offset
   Scale=@n10
   Offset_Pivot
   Rotation_Axis=NamedRerouteUsage()
  ) {materialFunction=&/MF_TopsSideProjection_Base_V2}
  n12 = ComponentMask(None=@n11.Combined) {a=true}
  n13 = OneMinus(None=Saturate(None=Add(A=@n6, B=@n12)))
  n14 = MaterialFunctionCall!(
   Normal_Map
   Offset Rt
   Pivot
   Caculated_Bounds=NamedRerouteUsage()
   Pre_Skinned_Normals=NamedRerouteUsage()
   is_State=NamedRerouteUsage()
   UV_Rotation=NamedRerouteUsage()
   Uvs_Offset
   Uvs_TileRate=NamedRerouteUsage()
   Uvs_Offset_Pivot
   For_Safe_Zones
   Rotation_Comp
   Clamp Y
   Clamp X
   Clamp Padding_Comp
   No_Rotation
  ) {materialFunction=&/MF_Projection_Side_Normal_V2}
  n15 = VectorParameter() {group=Wrap_RenegadeRaiderSpark_Noise, parameterName=Ren_Noise_Speed, defaultValue=<rgba 0.05 0.005 0 1>}
  n16 = Multiply(A=Time(), B=@n15.RGB)
  n17 = NamedRerouteUsage()
  n18 = TextureSample(UVs=Add(A=@n14.UVs, B=ComponentMask(None=@n16) {g=true, r=true}), Tex=@n17)
  n19 = NamedRerouteUsage()
  n20 = Multiply(
   A=OneMinus(None=Add(A=@n19) {constB=0.5})
   B=ScalarParameter() {defaultValue=1, group=Wrap_RenegadeRaiderSpark_Noise, parameterName=Ren_Curvature_Brightness}
  ) {constB=2}
  n21 = MaterialFunctionCall!(
   Bounds=NamedRerouteUsage()
   ForwardAxis=LinearInterpolate(
    A=NamedRerouteUsage()
    B=VectorParameter() {group=Wrap_RenegadeRaiderSpark, parameterName=Ren_IS_Roll_Foward, defaultValue=<rgba 0 0 0 1>}.RGB
    Alpha=ScalarParameter() {group=Wrap_RenegadeRaiderSpark, parameterName=Ren_IS_Roll}
   )
   Offset=ScalarParameter() {defaultValue=-0.1, group=Wrap_RenegadeRaiderSpark_Noise, parameterName=Ren_Grad_Offset}
   FallOff=ScalarParameter() {defaultValue=0.7, group=Wrap_RenegadeRaiderSpark_Noise, parameterName=Ren_Grad_Falloff}
  ) {materialFunction=&/MF_Wrap_Gradent_Simple}
  n22 = Multiply(A=Add(A=Multiply(A=@n13, B=@n18.A) {constA=1}, B=Saturate(None=@n20)), B=Power(Base=@n21.Raw_Data) {constExponent=1.25})
  n23 = CurveAtlasRowParameter(CurveTime=@n22) {
   atlas=/Game/Weapons/WeaponSkins/Wraps/Textures/CruveAtlas_Wraps_01
   curve=/Game/Weapons/WeaponSkins/Wraps/Textures/0_ColorCurves/CC_Ren_Fire
   defaultValue=95
   group=Wrap_RenegadeRaiderSpark_Noise
   parameterName=Ren_ColorCurve
  }
  n24 = MaterialFunctionCall!(
   Index=ScalarParameter() {defaultValue=95, group=Wrap_RenegadeRaiderSpark_Noise, parameterName="Ren_ColorCurve  Index"}
   CCinput=@n23
   CCinput_Alpha
   UV=@n22
  ) {materialFunction=&/MF_Global_CC_Tex}
  n25 = LinearInterpolate(
   A=ScalarParameter() {defaultValue=5, group=Wrap_RenegadeRaiderSpark_Noise, parameterName=Ren_BaseEmissive_Brightness_Start}
   B=ScalarParameter() {defaultValue=1, group=Wrap_RenegadeRaiderSpark_Noise, parameterName=Ren_BaseEmissive_Brightness_End}
   Alpha=NamedRerouteUsage()
  )
  Add(A=@n24.3V, B=@n25)
  ComponentMask() {g=true, r=true}
  ComponentMask() {g=true, r=true}
  Constant() {r=1}
  Constant() {r=1}
  n26 = BlendMaterialAttributes(A=NamedRerouteUsage(), B=NamedRerouteUsage(), Alpha=Saturate(None=@n2)) {vertexAttributeBlendType=UseA}
  n27 = NamedRerouteUsage()
  n28 = NamedRerouteDeclaration(None=@n3) {name=Msk_Accent, nodeColor=<rgba 0 1 0.7656 1>}
  n29 = TextureSample() {samplerType=SAMPLERTYPE_Normal, texture=/Game/Weapons/FORT_Rifles/T_SCAR_N}
  n30 = Constant3Vector() {constant=<rgba 1 0 0 0>}
  n31 = GetMaterialAttributes(None=FunctionInput(Preview=SetMaterialAttributes!(MaterialAttributes, Normal=@n29.RGB, BaseColor=@n30) {attributeSetTypes [#1, #2]}) {inputName=Material, inputType=FunctionInput_MaterialAttributes}) {attributeGetTypes [
    #2
    #3
    #1
    #4
   ]}
  FunctionOutput(None=StaticSwitchParameter(
    True=BlendMaterialAttributes(A=@n26, B=@n27, Alpha=@n28) {vertexAttributeBlendType=UseA}
    False=NamedRerouteDeclaration(None=@n31.MaterialAttributes) {name=Root_Material, nodeColor=<rgba 0.7812 1 0 1>}
   ) {group=Customization_Wrap, parameterName=Wrap_RenegadeRaiderSpark})
  n32 = ScalarParameter() {group=Wrap_RenegadeRaiderSpark_Uvs, parameterName=Ren_Rotation_Mech}
  n33 = MaterialFunctionCall(Float4=NamedRerouteUsage()) {materialFunction=&BreakOutFloat4Components}
  LinearInterpolate(A=@n32, B=ScalarParameter() {group=Wrap_RenegadeRaiderSpark_Uvs, parameterName=Ren_Rotation_Character}, Alpha=@n33.A)
  n34 = VectorParameter() {group=Wrap_RenegadeRaiderSpark_Uvs, parameterName=Ren_BCM_Rotation_Mech, defaultValue=<rgba 0.25 0.25 0 1>}
  n35 = VectorParameter() {group=Wrap_RenegadeRaiderSpark_Uvs, parameterName=Ren_BCM_Rotation_Character, defaultValue=<rgba 0.25 0.25 0 1>}
  n36 = MaterialFunctionCall(Float4=NamedRerouteUsage()) {materialFunction=&BreakOutFloat4Components}
  LinearInterpolate(A=@n34.RGB, B=@n35.RGB, Alpha=@n36.A)
  n37 = ScalarParameter() {group=Wrap_RenegadeRaiderSpark_Uvs, parameterName=Ren_Rotation_Vehicle}
  n38 = ScalarParameter() {group=Wrap_RenegadeRaiderSpark_Uvs, parameterName=Ren_Rotation_Vehicle_Side}
  n39 = LinearInterpolate(
   A=LinearInterpolate(
    A=ScalarParameter() {group=Wrap_RenegadeRaiderSpark_Uvs, parameterName=Ren_Rotation_Weapon}
    B=LinearInterpolate(A=@n37, B=@n38, Alpha=@n33.G)
    Alpha=@n33.R
   )
   B=@n32
   Alpha=@n33.B
  )
  n40 = MaterialFunctionCall(Float2=NamedRerouteUsage()) {materialFunction=&BreakOutFloat2Components}
  LinearInterpolate(
   A=LinearInterpolate(A=@n39, B=ScalarParameter() {group=Wrap_RenegadeRaiderSpark_Uvs, parameterName=Ren_Rotation_Backpack}, Alpha=@n40.R)
   B=ScalarParameter() {group=Wrap_RenegadeRaiderSpark_Uvs, parameterName=Ren_Rotation_Pickaxe}
   Alpha=@n40.G
  )
  n41 = VectorParameter() {group=Wrap_RenegadeRaiderSpark_Uvs, parameterName=Ren_BCM_Rotation_Weapon, defaultValue=<rgba 0 0 -0.25 1>}
  n42 = VectorParameter() {group=Wrap_RenegadeRaiderSpark_Uvs, parameterName=Ren_BCM_Rotation_Vehicle, defaultValue=<rgba 0 0 0 1>}
  n43 = VectorParameter() {group=Wrap_RenegadeRaiderSpark_Uvs, parameterName=Ren_BCM_Rotation_Vehicle_Side, defaultValue=<rgba 0 0 0 1>}
  n44 = LinearInterpolate(A=LinearInterpolate(A=@n41.RGB, B=LinearInterpolate(A=@n42.RGB, B=@n43.RGB, Alpha=@n36.G), Alpha=@n36.R), B=@n34.RGB, Alpha=@n36.B)
  n45 = VectorParameter() {group=Wrap_RenegadeRaiderSpark_Uvs, parameterName=Ren_BCM_Rotation_Backpack, defaultValue=<rgba 0 0 0 1>}
  n46 = MaterialFunctionCall(Float2=NamedRerouteUsage()) {materialFunction=&BreakOutFloat2Components}
  n47 = VectorParameter() {group=Wrap_RenegadeRaiderSpark_Uvs, parameterName=Ren_BCM_Rotation_Pickaxe, defaultValue=<rgba 0 0 0 1>}
  LinearInterpolate(A=LinearInterpolate(A=@n44, B=@n45.RGB, Alpha=@n46.R), B=@n47.RGB, Alpha=@n46.G)
  n48 = Constant3Vector() {constant=<rgba 0 0 -1 0>}
  n49 = MaterialFunctionCall(Float4=NamedRerouteUsage()) {materialFunction=&BreakOutFloat4Components}
  n50 = VectorParameter() {group=Wrap_RenegadeRaiderSpark_Crack, parameterName=Ren_Crack_Alternate}
  n51 = MaterialFunctionCall(
   Texture=NamedRerouteUsage()
   Caculated_Bounds=Add(A=NamedRerouteUsage(), B=LinearInterpolate(B=@n48, Alpha=@n49.B))
   Pre_Skinned_Normals=NamedRerouteUsage()
   Blend_Contrast=ScalarParameter() {defaultValue=2, group=Wrap_RenegadeRaiderSpark, parameterName=Ren_Tri_Contrast}
   IsStatsB=NamedRerouteUsage()
   IsState=NamedRerouteUsage()
   Fill_alternate_Axis_Opacity=Constant()
   "Fill_alternate_Axis_Color_A=Roughness"=AppendVector(A=@n50.RGB, B=@n50.RGB)
   Offset=NamedRerouteUsage()
   Scale=NamedRerouteUsage()
   Rotation_Axis=NamedRerouteUsage()
  ) {materialFunction=&MF_TopsSideProjection_Base_V2}
  Multiply(A=@n51.RGB)
  n52 = Fmod(A=@n4, B=Add(A=Constant() {r=1.5}))
  n53 = Constant() {r=2}
  n54 = Power(Base=Saturate(None=Divide(A=@n52, B=@n53))) {constExponent=0.5}
  Multiply(A=MaterialFunctionCall!(X range Y range, Gradient Height, ShowGrid?, Thickness, Color, "f(x)"=OneMinus(None=@n54)) {materialFunction=/Engine/Functions/Engine_MaterialFunctions01/Debug/PlotFunctionOnGraph}."f(x) = y")
  Multiply(A=MaterialFunctionCall(Value_Morning=Constant() {r=0.5}, Value_Day=Constant() {r=1}, Value_Evening=Constant() {r=0.5}, Value_Night=Constant() {r=0.25}) {materialFunction=&Scalar_Time_of_Day}.Result)
  n55 = VectorParameter() {group=Wrap_RenegadeRaiderSpark_Accent, parameterName=Ren_Accent_BaseColor, defaultValue=<rgba 0.005 0.005 0.005 1>}
  n56 = SetMaterialAttributes!(
   MaterialAttributes=NamedRerouteUsage()
   BaseColor=@n55.RGB
   Roughness=ScalarParameter() {defaultValue=0.6, group=Wrap_RenegadeRaiderSpark_Accent, parameterName=Ren_Accent_Roughness}
   Specular=ScalarParameter() {defaultValue=0.5, group=Wrap_RenegadeRaiderSpark_Accent, parameterName=Ren_Accent_Spec}
   Metallic=ScalarParameter() {group=Wrap_RenegadeRaiderSpark_Accent, parameterName=Ren_Accent_Metalic}
   Normal
   EmissiveColor
  ) {attributeSetTypes [
    #2
    #3
    #5
    #4
    #1
    #6
   ]}
  NamedRerouteDeclaration(None=@n56) {name=Mat_Accent, nodeColor=<rgba 0 0.3906 1 1>}
  n57 = FunctionInput(Preview=@n1.G) {bUsePreviewValueAsDefault=true, inputName=Mask Crevise, inputType=FunctionInput_Scalar, sortPriority=24, previewValue=<1 0 0 1>}
  NamedRerouteDeclaration(None=@n57) {name=Msk_Curvature, nodeColor=<rgba 0 1 0.7422 1>}
  NamedRerouteDeclaration(None=@n21.Raw_Data) {name=Gradent, nodeColor=<rgba 1 0 0.3047 1>}
  NamedRerouteDeclaration(None=TextureObjectParameter() {group=Wrap_RenegadeRaiderSpark, parameterName=Ren__Tex_MskC, texture=&T_BlacksRGB}) {name=Tex_MskC, nodeColor=<rgba 1 0.6694 0 1>}
  NamedRerouteDeclaration(None=@n2) {name=Msk_Prim, nodeColor=<rgba 0 1 0.7422 1>}
  n58 = TextureObjectParameter() {
   group=Wrap_RenegadeRaiderSpark
   parameterName=Ren__Tex_N
   samplerType=SAMPLERTYPE_Normal
   texture=&FlatNormalBlue
  }
  NamedRerouteDeclaration(None=@n58) {name=Tex_N, nodeColor=<rgba 1 0.6694 0 1>}
  NamedRerouteDeclaration(None=@n44) {name=Uvs_Rotation_BCM, nodeColor=<rgba 1 0 0.2578 1>}
  NamedRerouteDeclaration(None=FunctionInput(Preview=Constant3Vector() {constant=<rgba 0 1 0 0>}) {bUsePreviewValueAsDefault=true, inputName=ForwardAxis, sortPriority=100}) {name=ForwardAxis, nodeColor=<rgba 0 0.9297 1 1>}
  n59 = MaterialFunctionCall(Float4=@n51.Combined) {materialFunction=&BreakOutFloat4Components}
  n60 = NamedRerouteUsage()
  NamedRerouteDeclaration(None=Multiply(A=@n59.A, B=OneMinus(None=@n60)) {constB=0}) {name=MSk_Cracks_Alpha, nodeColor=<rgba 0.7344 0 1 1>}
  n61 = Multiply(A=@n5)
  n62 = NamedRerouteUsage()
  n63 = NamedRerouteUsage()
  n64 = Add(
   A=Add(A=Saturate(None=@n61), B=ComponentMask(None=@n11.Combined) {r=true})
   B=Multiply(A=Multiply(A=@n62) {constB=0.5}, B=OneMinus(None=@n63))
  )
  n65 = Saturate(None=Add(A=NamedRerouteUsage(), B=NamedRerouteUsage()))
  n66 = Add(A=Multiply(A=OneMinus(None=@n65), B=Multiply(A=@n24.3V, B=@n25) {constB=0.5}), B=NamedRerouteUsage())
  n67 = ScalarParameter() {defaultValue=10, group=Wrap_RenegadeRaiderSpark, parameterName=Maxx_Emissive_Brightness}
  n68 = NamedRerouteUsage()
  n69 = NamedRerouteUsage()
  n70 = Add(A=NamedRerouteUsage(), B=@n7)
  n71 = NamedRerouteUsage()
  n72 = NamedRerouteUsage()
  n73 = NamedRerouteUsage()
  n74 = Constant() {r=1}
  n75 = SetMaterialAttributes(
   MaterialAttributes=NamedRerouteUsage()
   BaseColor=VectorParameter() {group=Wrap_RenegadeRaiderSpark, parameterName=Ren_BaseColor, defaultValue=<rgba 0.01 0.01 0.01 1>}.RGB
   Roughness=LinearInterpolate(
    A=ScalarParameter() {group=Wrap_RenegadeRaiderSpark, parameterName=Ren_Roughness_High}
    B=ScalarParameter() {defaultValue=1, group=Wrap_RenegadeRaiderSpark, parameterName=Ren_Roughness_Low}
    Alpha=Saturate(None=@n64)
   )
   Specular=ScalarParameter() {defaultValue=0.5, group=Wrap_RenegadeRaiderSpark, parameterName=Ren_Spec}
   Metallic=ScalarParameter() {group=Wrap_RenegadeRaiderSpark, parameterName=Ren_Metalic}
   Normal=Add(
    A=NamedRerouteUsage()
    B=MaterialFunctionCall!(
     Texture_Normals=@n68
     Rotation_Axis=@n69
     Caculated_Bounds=@n70
     Pre_Skinned_Normals=@n71
     IsStatsB=@n72
     IsState=@n73
     Blend_Contrast=@n8
     Fill_alternate_Axis_Opacity=@n74
     Scale=@n10
     Offset_Pivot
     Offset
    ) {materialFunction=&/MF_TopsSideProjection_Normals_V2}.Add_Surface_Normal
   )
   EmissiveColor=LinearInterpolate(A=Saturate(None=@n66), B=Clamp(None=@n66, Max=@n67) {maxDefault=10}, Alpha=NamedRerouteUsage())
  ) {attributeSetTypes [
    #2
    #3
    #5
    #4
    #1
    #6
   ]}
  NamedRerouteDeclaration(None=@n75) {name=Mat_Base, nodeColor=<rgba 0 0.3906 1 1>}
  n76 = OneMinus(None=NamedRerouteUsage())
  NamedRerouteDeclaration(None=Multiply(A=@n59.R, B=@n76) {constB=0}) {name=MSk_Cracks_B, nodeColor=<rgba 0.7344 0 1 1>}
  NamedRerouteDeclaration(None=TextureObjectParameter() {group=Wrap_RenegadeRaiderSpark, parameterName=Ren__Tex_MskA, texture=&T_BlacksRGB}) {name=Tex_MskA, nodeColor=<rgba 1 0.6694 0 1>}
  n77 = Add(A=NamedRerouteUsage()) {constB=-0.5}
  n78 = Power(Base=Multiply(A=NamedRerouteUsage()) {constB=0.8}) {constExponent=10}
  n79 = LinearInterpolate(A=Multiply(A=NamedRerouteUsage()) {constB=1.2}, Alpha=@n78)
  n80 = Multiply(A=Multiply(A=OneMinus(None=@n77), B=Multiply(A=@n79, B=@n59.R)), B=@n76)
  n81 = CurveAtlasRowParameter(CurveTime=@n80) {
   atlas=&CruveAtlas_Wraps_01
   curve=/Game/Weapons/WeaponSkins/Wraps/Textures/0_ColorCurves/CC_Ren_FireB
   defaultValue=96
   group=Wrap_RenegadeRaiderSpark_Crack
   parameterName=Ren_ColorCurve_Cracks
  }
  n82 = MaterialFunctionCall(
   Index=ScalarParameter() {defaultValue=96, group=Wrap_RenegadeRaiderSpark_Crack, parameterName="Ren_ColorCurve_Cracks  Index"}
   CCinput=@n81
   UV=@n80
  ) {materialFunction=&MF_Global_CC_Tex}
  n83 = Multiply(A=@n82.3V, B=ScalarParameter() {defaultValue=5, group=Wrap_RenegadeRaiderSpark_Crack, parameterName=Ren_CC_Crack_Brightness}) {constB=5}
  n84 = ScalarParameter() {defaultValue=5, group=Wrap_RenegadeRaiderSpark_Crack, parameterName=Ren_Crack_Spark_Brightness_Start}
  n85 = NamedRerouteUsage()
  NamedRerouteDeclaration(None=Multiply(A=LinearInterpolate(A=@n82.3V, B=@n83, Alpha=@n78), B=LinearInterpolate(A=@n84, Alpha=@n85)) {constB=0}) {name=MSk_Cracks_Glow, nodeColor=<rgba 0.7344 0 1 1>}
  n86 = Multiply(A=@n14.UVs, B=ScalarParameter() {defaultValue=7, group=Wrap_RenegadeRaiderSpark_Noise, parameterName=Ren_Noise_TileRate_B}) {constB=0.5}
  n87 = VectorParameter() {group=Wrap_RenegadeRaiderSpark_Noise, parameterName=Ren_Noise_Speed_B, defaultValue=<rgba 0.5 0.25 0 1>}
  n88 = ComponentMask(None=Multiply(A=Time(), B=@n87.RGB)) {g=true, r=true}
  NamedRerouteDeclaration(None=TextureSample(UVs=Add(A=@n86, B=@n88), Tex=@n17).R) {name=Perlin_Noise, nodeColor=<rgba 0 0.2734 1 1>}
  n89 = ScalarParameter() {defaultValue=1.5, group=Wrap_RenegadeRaiderSpark_Uvs, parameterName=Ren_BCM_Scale_Vehicle}
  n90 = ScalarParameter() {defaultValue=1.5, group=Wrap_RenegadeRaiderSpark_Uvs, parameterName=Ren_BCM_Scale_Vehicle_Side}
  n91 = MaterialFunctionCall(Float4=NamedRerouteUsage()) {materialFunction=&BreakOutFloat4Components}
  n92 = LinearInterpolate(
   A=LinearInterpolate(
    A=ScalarParameter() {defaultValue=1, group=Wrap_RenegadeRaiderSpark_Uvs, parameterName=Ren_BCM_Scale_Weapon}
    B=LinearInterpolate(A=@n89, B=@n90, Alpha=@n91.G)
    Alpha=@n91.R
   )
   B=LinearInterpolate(
    A=ScalarParameter() {defaultValue=1.5, group=Wrap_RenegadeRaiderSpark_Uvs, parameterName=Ren_BCM_Ren_BCM_Scale_Mech}
    B=ScalarParameter() {defaultValue=1.5, group=Wrap_RenegadeRaiderSpark_Uvs, parameterName=Ren_BCM_Scale_Character}
    Alpha=@n91.A
   )
   Alpha=@n91.B
  )
  n93 = ScalarParameter() {defaultValue=1.5, group=Wrap_RenegadeRaiderSpark_Uvs, parameterName=Ren_BCM_Scale_Backpack}
  n94 = MaterialFunctionCall(Float2=NamedRerouteUsage()) {materialFunction=&BreakOutFloat2Components}
  NamedRerouteDeclaration(None=LinearInterpolate(
    A=LinearInterpolate(A=@n92, B=@n93, Alpha=@n94.R)
    B=ScalarParameter() {defaultValue=1.5, group=Wrap_RenegadeRaiderSpark_Uvs, parameterName=Ren_BCM_Scale_Pickaxe}
    Alpha=@n94.G
   )) {name=UVs_Scale_BCM, nodeColor=<rgba 0 0.7188 1 1>}
  n95 = MaterialFunctionCall(Color=Abs(None=NamedRerouteUsage())) {materialFunction=&SplitComponents}
  n96 = MaterialFunctionCall(Float4=NamedRerouteUsage()) {materialFunction=&BreakOutFloat4Components}
  n97 = LinearInterpolate(A=@n95.R, B=LinearInterpolate(A=@n95.B, B=@n95.B, Alpha=@n96.G), Alpha=@n96.R)
  NamedRerouteDeclaration(None=MaterialFunctionCall(
    In=LinearInterpolate(A=@n97, B=@n95.G, Alpha=@n96.B)
    Contrast=ScalarParameter() {defaultValue=1, group=Wrap_RenegadeRaiderSpark, parameterName=Ren_Blend_Contrast}
   ) {materialFunction=&CheapContrast}.Result) {name=Tri_Mask, nodeColor=<rgba 1 0.4688 0 1>}
  NamedRerouteDeclaration(None=@n39) {name=UVs_Rotation, nodeColor=<rgba 0 0.7188 1 1>}
  n98 = ScalarParameter() {defaultValue=1, group=Wrap_RenegadeRaiderSpark_Uvs, parameterName=Ren_Crack_Scale_Vehicle}
  n99 = ScalarParameter() {defaultValue=1, group=Wrap_RenegadeRaiderSpark_Uvs, parameterName=Ren_Crack_Scale_Vehicle_Side}
  n100 = MaterialFunctionCall(Float4=NamedRerouteUsage()) {materialFunction=&BreakOutFloat4Components}
  n101 = LinearInterpolate(
   A=LinearInterpolate(
    A=ScalarParameter() {defaultValue=1, group=Wrap_RenegadeRaiderSpark_Uvs, parameterName=Ren_Crack_Scale_Weapon}
    B=LinearInterpolate(A=@n98, B=@n99, Alpha=@n100.G)
    Alpha=@n100.R
   )
   B=LinearInterpolate(
    A=ScalarParameter() {defaultValue=1, group=Wrap_RenegadeRaiderSpark_Uvs, parameterName=Ren_Crack_Ren_BCM_Scale_Mech}
    B=ScalarParameter() {defaultValue=1, group=Wrap_RenegadeRaiderSpark_Uvs, parameterName=Ren_Crack_Scale_Character}
    Alpha=@n100.A
   )
   Alpha=@n100.B
  )
  n102 = ScalarParameter() {defaultValue=1, group=Wrap_RenegadeRaiderSpark_Uvs, parameterName=Ren_Crack_Scale_Backpack}
  n103 = MaterialFunctionCall(Float2=NamedRerouteUsage()) {materialFunction=&BreakOutFloat2Components}
  NamedRerouteDeclaration(None=LinearInterpolate(
    A=LinearInterpolate(A=@n101, B=@n102, Alpha=@n103.R)
    B=ScalarParameter() {defaultValue=1, group=Wrap_RenegadeRaiderSpark_Uvs, parameterName=Ren_Crack_Scale_Pickaxe}
    Alpha=@n103.G
   )) {name=UVs_Scale_Crack, nodeColor=<rgba 0 0.7188 1 1>}
  n104 = VectorParameter() {group=Wrap_RenegadeRaiderSpark_Uvs, parameterName=Ren_Crack_Offset_Weapon, defaultValue=<rgba 0 0 0 1>}
  n105 = VectorParameter() {group=Wrap_RenegadeRaiderSpark_Uvs, parameterName=Ren_Crack_Offset_Vehicle, defaultValue=<rgba 0 0 0 1>}
  n106 = VectorParameter() {group=Wrap_RenegadeRaiderSpark_Uvs, parameterName=Ren_Crack_Offset_VehicleSide, defaultValue=<rgba 0 0 0 1>}
  n107 = MaterialFunctionCall(Float4=NamedRerouteUsage()) {materialFunction=&BreakOutFloat4Components}
  n108 = LinearInterpolate(A=@n105.RGB, B=@n106.RGB, Alpha=@n107.G)
  n109 = VectorParameter() {group=Wrap_RenegadeRaiderSpark_Uvs, parameterName=Ren_Crack_Offset_Mech, defaultValue=<rgba 0 0 0 1>}
  n110 = VectorParameter() {group=Wrap_RenegadeRaiderSpark_Uvs, parameterName=Ren_Crack_Offset_Character, defaultValue=<rgba 0 0 0 1>}
  n111 = VectorParameter() {group=Wrap_RenegadeRaiderSpark_Uvs, parameterName=Ren_Crack_Offset_Backpack, defaultValue=<rgba 0 0 0 1>}
  n112 = MaterialFunctionCall(Float2=NamedRerouteUsage()) {materialFunction=&BreakOutFloat2Components}
  n113 = LinearInterpolate(
   A=LinearInterpolate(A=LinearInterpolate(A=@n104.RGB, B=@n108, Alpha=@n107.R), B=LinearInterpolate(A=@n109.RGB, B=@n110.RGB, Alpha=@n107.A), Alpha=@n107.B)
   B=@n111.RGB
   Alpha=@n112.R
  )
  n114 = VectorParameter() {group=Wrap_RenegadeRaiderSpark_Uvs, parameterName=Ren_Crack_Offset_Pickaxe, defaultValue=<rgba 0 0 0 1>}
  n115 = AppendVector(A=Constant(), B=ScalarParameter() {group=Wrap_RenegadeRaiderSpark_Uvs, parameterName=Ren_Crack_Offset})
  n116 = Constant()
  NamedRerouteDeclaration(None=Add(A=LinearInterpolate(A=@n113, B=@n114.RGB, Alpha=@n112.G), B=AppendVector(A=@n115, B=@n116))) {name=UVs_Offset_Cracks, nodeColor=<rgba 0 0.625 1 1>}
  n117 = ScalarParameter() {defaultValue=-9999999, group=Wrap_RenegadeRaiderSpark_Reactive, parameterName=Ren_TimeFrom-BP----------------, sortPriority=600}
  n118 = Subtract(A=@n4, B=@n117)
  n119 = ScalarParameter() {defaultValue=0.25, group=Wrap_RenegadeRaiderSpark_Reactive, parameterName=Ren_Spark_Duration}
  NamedRerouteDeclaration(None=Saturate(None=Divide(A=@n118, B=@n119))) {name=Time_BP_Spark, nodeColor=<rgba 0 0.8359 1 1>}
  n120 = ScalarParameter() {defaultValue=1, group=Wrap_RenegadeRaiderSpark_Reactive, parameterName=Ren_Crack_Delay}
  n121 = Divide(
   A=Subtract(A=@n4, B=Add(A=@n117, B=@n120))
   B=ScalarParameter() {defaultValue=0.25, group=Wrap_RenegadeRaiderSpark_Reactive, parameterName=Ren_Crack_Duration}
  )
  NamedRerouteDeclaration(None=Power(
    Base=Saturate(None=@n121)
    Exp=ScalarParameter() {defaultValue=0.5, group=Wrap_RenegadeRaiderSpark_Reactive, parameterName=Ren_Crack_Falloff}
   )) {name=Time_BP_Crack, nodeColor=<rgba 0 0.8359 1 1>}
  n122 = ScalarParameter() {defaultValue=1, group=Wrap_RenegadeRaiderSpark_Reactive, parameterName=Ren_Crack_Soot_Delay}
  n123 = Divide(
   A=Subtract(A=@n4, B=Add(A=@n117, B=@n122))
   B=ScalarParameter() {defaultValue=0.25, group=Wrap_RenegadeRaiderSpark_Reactive, parameterName=Ren_Crack_Soot_Duration}
  )
  NamedRerouteDeclaration(None=Power(
    Base=Saturate(None=@n123)
    Exp=ScalarParameter() {defaultValue=0.5, group=Wrap_RenegadeRaiderSpark_Reactive, parameterName=Ren_Crack_Soot_Falloff}
   )) {name=Time_BP_Crack_Soot, nodeColor=<rgba 0 0.8359 1 1>}
  n124 = ScalarParameter() {defaultValue=1, group=Wrap_RenegadeRaiderSpark_Uvs, parameterName=Ren_Noise_Scale_Vehicle}
  n125 = ScalarParameter() {defaultValue=1, group=Wrap_RenegadeRaiderSpark_Uvs, parameterName=Ren_Noise_Scale_Vehicle_Side}
  n126 = MaterialFunctionCall(Float4=NamedRerouteUsage()) {materialFunction=&BreakOutFloat4Components}
  n127 = LinearInterpolate(
   A=LinearInterpolate(
    A=ScalarParameter() {defaultValue=1, group=Wrap_RenegadeRaiderSpark_Uvs, parameterName=Ren_Noise_Scale_Weapon}
    B=LinearInterpolate(A=@n124, B=@n125, Alpha=@n126.G)
    Alpha=@n126.R
   )
   B=LinearInterpolate(
    A=ScalarParameter() {defaultValue=1, group=Wrap_RenegadeRaiderSpark_Uvs, parameterName=Ren_Noise_Scale_Mech}
    B=ScalarParameter() {defaultValue=1, group=Wrap_RenegadeRaiderSpark_Uvs, parameterName=Ren_Noise_Scale_Character}
    Alpha=@n126.A
   )
   Alpha=@n126.B
  )
  n128 = ScalarParameter() {defaultValue=1, group=Wrap_RenegadeRaiderSpark_Uvs, parameterName=Ren_Noise_Scale_Backpack}
  n129 = MaterialFunctionCall(Float2=NamedRerouteUsage()) {materialFunction=&BreakOutFloat2Components}
  NamedRerouteDeclaration(None=LinearInterpolate(
    A=LinearInterpolate(A=@n127, B=@n128, Alpha=@n129.R)
    B=ScalarParameter() {defaultValue=1, group=Wrap_RenegadeRaiderSpark_Uvs, parameterName=Ren_Noise_Scale_Pickaxe}
    Alpha=@n129.G
   )) {name=UVs_Scale_Noise, nodeColor=<rgba 0 0.7188 1 1>}
  NamedRerouteDeclaration(None=FunctionInput(Preview=Constant4Vector() {constant=<rgba 0 0 1 0>}.RGBA) {bUsePreviewValueAsDefault=true, inputName=IsState, inputType=FunctionInput_Vector4, sortPriority=4}) {name=IsState, nodeColor=<rgba 0.1497 1 0.1633 1>}
  NamedRerouteDeclaration(None=FunctionInput(Preview=Constant2Vector()) {bUsePreviewValueAsDefault=true, inputName=IsState_B, inputType=FunctionInput_Vector2, sortPriority=4}) {name=IsState_B, nodeColor=<rgba 0.1497 1 0.1633 1>}
  n130 = PreSkinnedLocalBounds()
  n131 = Subtract(A=VertexInterpolator(VS=PreSkinnedPosition()).PS, B=@n130.Min)
  n132 = MaterialFunctionCall(Color=Subtract(A=@n130.Max, B=@n130.Max)) {materialFunction=&SplitComponents}
  n133 = Max(A=Max(A=@n132.R, B=@n132.R), B=@n132.B)
  NamedRerouteDeclaration(None=FunctionInput(Preview=Divide(A=@n131, B=@n133)) {bUsePreviewValueAsDefault=true, inputName=Calcated_Bounds, sortPriority=11}) {name=Caclcated_Bounds, nodeColor=<rgba 0 0.9297 1 1>}
  NamedRerouteDeclaration(None=TextureObjectParameter() {group=Wrap_RenegadeRaiderSpark, parameterName=Ren__Tex_MskB, texture=&T_BlacksRGB}) {name=Tex_MskB, nodeColor=<rgba 1 0.6694 0 1>}
  n134 = PreSkinnedNormal()
  NamedRerouteDeclaration(None=FunctionInput(Preview=VertexInterpolator(VS=@n134).PS) {bUsePreviewValueAsDefault=true, inputName=Pre-Skinned Local Normal, sortPriority=10}) {name=PreSkin_Normal, nodeColor=<rgba 0.8281 1 0 1>}
  NamedRerouteDeclaration(None=@n31.Normal) {name=Root_Normal, nodeColor=<rgba 0.4062 0 1 1>}
  ScalarParameter() {defaultValue=1, group=Wrap_RenegadeRaiderSpark_Noise, parameterName=Ren_Noise_TileRate}
  ScalarParameter() {defaultValue=2, group=Wrap_RenegadeRaiderSpark_Crack, parameterName=Ren_Crack_Brightness_Start}
 ]}
