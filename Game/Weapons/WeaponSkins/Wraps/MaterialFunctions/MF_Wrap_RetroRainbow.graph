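# Material function for the RetroRainbow wrap skin, blending a projected rainbow design into customization-masked areas of weapons, vehicles, mechs and characters behind a static switch.
# IsState selects per-type projection, tiling and offsets, while world-aligned noise varies specular, roughness and optionally color.
# Animated flipbook diamonds, randomized per grid cell and masked by normal direction, appear on a Blueprint timestamp and fade out.
# The gradient-diamond and second base-color branches are unconnected.
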MaterialFunction /Game/Weapons/WeaponSkins/Wraps/MaterialFunctions/MF_Wrap_RetroRainbow
graph {nodes [
  n1 = TextureObjectParameter() {group=Wrap__RetroRainbow_Textures, parameterName=RR_Mask_02, sortPriority=20, texture=&T_BlacksRGB}
  n2 = MaterialFunctionCall(Color=Constant3Vector() {constant=<rgba 0 -0.25 -0.25 1>}) {materialFunction=&SplitComponents}
  n3 = VertexInterpolator(VS=PreSkinnedPosition())
  n4 = PreSkinnedLocalBounds()
  n5 = FunctionInput(Preview=Divide(A=Subtract(A=@n3.PS, B=@n4.Min), B=Subtract(A=@n4.Max, B=@n4.Max))) {bUsePreviewValueAsDefault=true, inputName=Calcated_Bounds, sortPriority=11}
  n6 = FunctionInput(Preview=VertexInterpolator(VS=PreSkinnedNormal()).PS) {bUsePreviewValueAsDefault=true, inputName=Pre-Skinned Local Normal, sortPriority=10}
  n7 = MaterialFunctionCall(
   Export Float 4=StaticBool() {value=true}
   TextureObject=@n1
   TextureSize=Constant() {r=1}
   WorldPosition=@n5
   World Space Normal=@n6
   RotZAxis=@n2.B
   RotXAxis=@n2.B
   RotYAxis=@n2.B
   ProjectionTransitionContrast=ScalarParameter() {defaultValue=1, group=Wrap_RetroRainbow, parameterName=RR_HFNoise_Blend_Contrast, sortPriority=-2}
  ) {materialFunction=&/WorldAlignedTexture}
  n8 = ChannelMaskParameter(None=@n7.XYZ Texture) {group=Wrap_RetroRainbow, maskChannel=Alpha, parameterName=RR_HFNoise__ChannelMask-----, sortPriority=-15, defaultValue=<rgba 0 0 0 1>}
  n9 = OneMinus(None=@n8.R)
  n10 = TextureObjectParameter() {group=Wrap__RetroRainbow_Textures, parameterName=RR_BaseColor_01, sortPriority=11, texture=&T_BlacksRGB}
  n11 = FunctionInput() {inputName=IsState, inputType=FunctionInput_Vector4, sortPriority=5}
  n12 = MaterialFunctionCall(Color=@n11) {materialFunction=&SplitComponents}
  n13 = VectorParameter() {group=Wrap_RetroRainbow_UVs, parameterName=RR_Offset_Weapon, defaultValue=<rgba 0 0 0 1>}
  n14 = MaterialFunctionCall(Float4=@n11) {materialFunction=&BreakOutFloat4Components}
  n15 = LinearInterpolate(
   A=VectorParameter() {group=Wrap_RetroRainbow_UVs, parameterName=RR_Offset_Vehicle, defaultValue=<rgba 0 0 0 1>}.RGB
   B=VectorParameter() {group=Wrap_RetroRainbow_UVs, parameterName=RR_Offset_Vehicle_Side, defaultValue=<rgba 0 0 0 1>}.RGB
   Alpha=@n14.G
  )
  n16 = VectorParameter() {group=Wrap_RetroRainbow_UVs, parameterName=RR_Offset_Mech -----, defaultValue=<rgba 0 0 0 1>}
  n17 = VectorParameter() {group=Wrap_RetroRainbow_UVs, parameterName=RR_Offset_Character, defaultValue=<rgba 0 0 0 1>}
  n18 = ScalarParameter() {defaultValue=1, group=Wrap_RetroRainbow_UVs, parameterName=RR_TileRate_Weapon}
  n19 = LinearInterpolate(
   A=ScalarParameter() {defaultValue=1, group=Wrap_RetroRainbow_UVs, parameterName=RR_TileRate_Vehicle}
   B=ScalarParameter() {defaultValue=1, group=Wrap_RetroRainbow_UVs, parameterName=RR_TileRate_VehicleSide}
   Alpha=@n14.G
  )
  n20 = ScalarParameter() {defaultValue=1, group=Wrap_RetroRainbow_UVs, parameterName=RR_TileRate_Mech -----}
  n21 = ScalarParameter() {defaultValue=1, group=Wrap_RetroRainbow_UVs, parameterName=RR_TileRate_Character}
  n22 = MaterialFunctionCall!(
   For_Safe_Zones
   Normal_Map
   Caculated_Bounds=@n5
   Pre_Skinned_Normals=@n6
   Is Vehicle=@n12.R
   Is VehicleSide=@n12.R
   Is Mech=@n12.R
   Vehicle_Rotation=ScalarParameter() {group=Wrap_RetroRainbow_UVs, parameterName=RR_Rotation_Vehicle}
   UV_Rotation Bypass
   Mech_Rotation=ScalarParameter() {defaultValue=0.25, group=Wrap_RetroRainbow_UVs, parameterName=RR_Rotation_Mech -----}
   Uvs_Offset=LinearInterpolate(A=LinearInterpolate(A=@n13.RGB, B=@n15, Alpha=@n14.R), B=LinearInterpolate(A=@n16.RGB, B=@n17.RGB, Alpha=@n14.A), Alpha=@n14.B)
   Offset-after rotation
   Uvs_TileRate=LinearInterpolate(A=LinearInterpolate(A=@n18, B=@n19, Alpha=@n14.R), B=LinearInterpolate(A=@n20, B=@n21, Alpha=@n14.A), Alpha=@n14.B)
  ) {materialFunction=&/MF_Projection_Side_Normal}
  n23 = TextureSample(UVs=@n22.UVs, Tex=@n10)
  n24 = LinearInterpolate(
   A=VectorParameter() {group=Wrap_RetroRainbow, parameterName=RR_Diamond_BGSpeed, defaultValue=<rgba 0.05 0 0 1>}.RGB
   B=VectorParameter() {group=Wrap_RetroRainbow, parameterName=RR_Diamond_Character_BGSpeed, defaultValue=<rgba 0.05 0 0 1>}.RGB
   Alpha=NamedRerouteUsage()
  )
  n25 = Add(A=@n22.UVs No Offsets/Tiling, B=Multiply(A=ComponentMask(None=@n24) {g=true, r=true}, B=Time()))
  n26 = ScalarParameter() {defaultValue=16, group=Wrap_RetroRainbow, parameterName=RR_Diamond_TileRate_Vehicle, sortPriority=100}
  n27 = ScalarParameter() {defaultValue=16, group=Wrap_RetroRainbow, parameterName=RR_Diamond_TileRate_VehicleSide, sortPriority=100}
  n28 = MaterialFunctionCall(Color=@n11) {materialFunction=&SplitComponents}
  n29 = LinearInterpolate(
   A=LinearInterpolate(
    A=ScalarParameter() {defaultValue=16, group=Wrap_RetroRainbow, parameterName=RR_Diamond_TileRate_Weapon, sortPriority=100}
    B=LinearInterpolate(A=@n26, B=@n27, Alpha=@n28.G)
    Alpha=@n28.R
   )
   B=ScalarParameter() {defaultValue=16, group=Wrap_RetroRainbow, parameterName=RR_Diamond_TileRate_Mech-----, sortPriority=100}
   Alpha=@n28.B
  )
  n30 = ComponentMask(None=@n25) {r=true}
  n31 = Divide(A=Floor(None=Multiply(A=@n30, B=@n29) {constB=5}), B=@n29) {constB=5}
  n32 = ScalarParameter() {defaultValue=4, group=Wrap_RetroRainbow, parameterName=RR_Diamond_Offset_Y-----, sortPriority=40}
  n33 = Add(A=Multiply(A=@n25, B=@n29), B=AppendVector(A=Constant(), B=Multiply(A=@n31, B=@n32))) {constB=0}
  n34 = VectorParameter() {group=Wrap_RetroRainbow, parameterName=RR_Diamond_Random_seed, sortPriority=40, defaultValue=<rgba 4 3 0 0>}
  n35 = MaterialFunctionCall!(2 Vector=Add(A=Floor(None=@n33), B=ComponentMask(None=@n34.RGB) {g=true, r=true}), RandomSeed) {materialFunction=&/MF_Random_2}
  n36 = Constant() {r=3}
  n37 = TextureObjectParameter() {group=Wrap__RetroRainbow_Textures, parameterName=RR_Mask_03, sortPriority=20, texture=&T_BlacksRGB}
  n38 = ScalarParameter() {defaultValue=0.7, group=Wrap_RetroRainbow, parameterName=RR_Diamond_Scale_Speed, sortPriority=10}
  n39 = Sine(None=Multiply(A=Time(), B=Multiply(A=@n38, B=@n35.utput)))
  n40 = ScalarParameter() {defaultValue=0.35, group=Wrap_RetroRainbow, parameterName=RR_Diamond_Scale_Sine_Amount, sortPriority=10}
  n41 = MaterialFunctionCall(
   UVs=Frac(None=@n33)
   Texture Scale=Add(
    A=ScalarParameter() {defaultValue=0.5, group=Wrap_RetroRainbow, parameterName=RR_Diamond_Scale-----, sortPriority=10}
    B=Multiply(A=@n39, B=@n40)
   ) {constA=0.8}
  ) {materialFunction=&ScaleUVsByCenter}
  n42 = MaterialFunctionCall(
   "Animation  Phase (0-1)"=LinearInterpolate(Alpha=@n35.utput) {constA=-3, constB=3}
   Number of Rows=@n36
   Number of Columns=@n36
   Texture=@n37
   UVs=@n41.UVs
  ) {materialFunction=&FlipBook}
  n43 = Constant() {r=0.5}
  n44 = Constant() {r=0.5}
  n45 = Constant() {r=0.9}
  n46 = Add(A=@n35.utput, B=Multiply(A=Time(), B=ScalarParameter() {defaultValue=0.05, group=Wrap_RetroRainbow, parameterName=RR_Diamond_BGFade_Speed}))
  n47 = Saturate(None=LinearInterpolate(Alpha=Sine(None=@n46)) {constA=0.4, constB=1.25})
  n48 = Multiply(A=@n42.Alpha, B=Multiply(A=SphereMask(A=@n41.UVs, B=@n43, Radius=@n44, Hardness=@n45), B=Round(None=@n47)))
  n49 = MaterialFunctionCall(Color=Abs(None=@n6)) {materialFunction=&SplitComponents}
  n50 = MaterialFunctionCall(Color=@n11) {materialFunction=&SplitComponents}
  n51 = LinearInterpolate(A=@n49.R, B=LinearInterpolate(A=@n49.B, B=@n49.B, Alpha=@n50.G), Alpha=@n50.R)
  n52 = Saturate(None=MaterialFunctionCall(
    In=LinearInterpolate(A=@n51, B=@n49.G, Alpha=@n50.B)
    Contrast=ScalarParameter() {defaultValue=5, group=Wrap_RetroRainbow, parameterName=RR_DiamondMask_Contrast}
   ) {materialFunction=&CheapContrast}.Result)
  n53 = Multiply(A=Multiply(A=ComponentMask(None=@n42.Result) {b=true, g=true, r=true}, B=@n48), B=@n52)
  n54 = Multiply(A=@n52, B=@n48)
  n55 = Multiply(A=ScalarParameter() {defaultValue=0.2, group=Wrap_RetroRainbow, parameterName=RR_Diamonds_StripeColorMultiplier}, B=@n54)
  n56 = Time()
  n57 = ScalarParameter() {defaultValue=-9999999, group=Wrap_RetroRainbow_Reactive, parameterName=TimeFrom-BP----------------, sortPriority=-600}
  n58 = Saturate(None=Divide(
    A=Subtract(A=@n56, B=@n57)
    B=ScalarParameter() {defaultValue=1.5, group=Wrap_RetroRainbow_Reactive, parameterName=RR_AnimationDuration, sortPriority=33}
   ))
  n59 = ScalarParameter() {defaultValue=1.1, group=Wrap_RetroRainbow_Reactive, parameterName=RR_Curve_Amount, sortPriority=34}
  n60 = ScalarParameter() {group=Wrap_RetroRainbow_Reactive, parameterName=RR_TestReactive, sortPriority=300}
  n61 = LinearInterpolate(A=OneMinus(None=Power(Base=@n58, Exp=@n59)) {desc=inverts curve from 0-1 to 1-0}, B=@n60, Alpha=@n60)
  n62 = LinearInterpolate(A=@n23.RGB, B=LinearInterpolate(A=LinearInterpolate(A=@n23.RGB, B=@n53, Alpha=@n54), B=Add(A=@n23.RGB, B=@n55), Alpha=@n23.A), Alpha=@n61)
  n63 = Multiply(
   A=@n62
   B=Power(
    Base=@n9
    Exp=ScalarParameter() {defaultValue=0.25, group=Wrap_RetroRainbow, parameterName=RR_HFNoise_BaseColor_MultiplyBlend, sortPriority=-12}
   )
  )
  n64 = Multiply(A=@n8.R, B=ScalarParameter() {defaultValue=0.025, group=Wrap_RetroRainbow, parameterName=RR_HFNoise_BaseColor_AddBlend, sortPriority=-12})
  n65 = ScalarParameter() {defaultValue=0.5, group=Wrap_RetroRainbow, parameterName=RR_Specular, sortPriority=3}
  n66 = LinearInterpolate(
   A=@n65
   B=LinearInterpolate(A=@n65, B=ScalarParameter() {group=Wrap_RetroRainbow, parameterName=RR_Specular_Diamond, sortPriority=3}, Alpha=@n54)
   Alpha=@n61
  )
  n67 = Multiply(A=@n9, B=ScalarParameter() {defaultValue=1.25, group=Wrap_RetroRainbow, parameterName=RR_HFNoise_MultiplyBlend, sortPriority=-10})
  n68 = Multiply(A=@n8.R, B=ScalarParameter() {defaultValue=0.035, group=Wrap_RetroRainbow, parameterName=RR_HFNoise_AddBlend, sortPriority=-10})
  n69 = LinearInterpolate(
   A=ScalarParameter() {group=Wrap_RetroRainbow, parameterName=RR_Roughness, sortPriority=2}
   B=ScalarParameter() {defaultValue=1, group=Wrap_RetroRainbow, parameterName=RR_Roughness_High, sortPriority=2}
   Alpha=ChannelMaskParameter(None=@n7.XYZ Texture) {group=Wrap_RetroRainbow, parameterName=RR_Roughness__ChannelMask-----, sortPriority=1}.R
  )
  n70 = LinearInterpolate(
   A=@n69
   B=LinearInterpolate(A=@n69, B=ScalarParameter() {defaultValue=1, group=Wrap_RetroRainbow, parameterName=RR_Roughness_Diamond, sortPriority=2}, Alpha=@n54)
   Alpha=@n61
  )
  n71 = GetMaterialAttributes(None=FunctionInput() {inputName=Material, inputType=FunctionInput_MaterialAttributes, sortPriority=-20}) {attributeGetTypes [#1]}
  n72 = SetMaterialAttributes!(
   MaterialAttributes=@n71.MaterialAttributes
   BaseColor=LinearInterpolate(A=@n62, B=Add(A=@n63, B=@n64), Alpha=ScalarParameter() {group=Wrap_RetroRainbow, parameterName=RR_UseBaseColor_Blends, sortPriority=-13})
   Metallic=ScalarParameter() {group=Wrap_RetroRainbow, parameterName=RR_Metallic, sortPriority=5}
   Specular=Add(A=Multiply(A=@n66, B=@n67), B=@n68)
   Roughness=Add(A=Multiply(A=@n70, B=@n67), B=@n68)
   Normal=@n71.Normal
  ) {attributeSetTypes [
    #2
    #3
    #4
    #5
    #1
   ]}
  n73 = FunctionInput() {
   bUsePreviewValueAsDefault=true
   desc=base customization
   inputName="Customization Mask (_CM B Top left)"
   inputType=FunctionInput_Scalar
   sortPriority=49
  }
  n74 = FunctionInput() {
   bUsePreviewValueAsDefault=true
   desc=secondary
   inputName="Customization Mask 2 (_CM B Top Right)"
   inputType=FunctionInput_Scalar
   sortPriority=50
  }
  n75 = NamedRerouteUsage()
  n76 = NamedRerouteUsage()
  n77 = LinearInterpolate(A=Add(A=@n73, B=@n74), B=@n73, Alpha=Saturate(None=Subtract(A=@n75, B=@n76)))
  FunctionOutput(None=StaticSwitchParameter(True=BlendMaterialAttributes(A=@n71.MaterialAttributes, B=@n72, Alpha=@n77) {vertexAttributeBlendType=UseA}, False=@n71.MaterialAttributes) {group=Customization_Wrap, parameterName=Wrap_RetroRainbow})
  NamedRerouteDeclaration(None=MaterialFunctionCall(Float4=@n11) {materialFunction=&BreakOutFloat4Components}.A) {name=IsCharacter, nodeColor=<rgba 1 0.8378 0 1>}
  NamedRerouteDeclaration(None=FunctionInput() {bUsePreviewValueAsDefault=true, inputName=isCharAccent, inputType=FunctionInput_Scalar, sortPriority=78}) {name=isCharAccent, nodeColor=<rgba 0 0.9253 1 1>}
  n78 = Time()
  n79 = ScalarParameter() {defaultValue=-0.25, group=Wrap_RetroRainbow, parameterName=RR_Diamonds01_GradientSpeed, sortPriority=-26}
  n80 = Multiply(
   A=Add(A=Multiply(A=@n78, B=@n79), B=@n22.UVs)
   B=ScalarParameter() {defaultValue=1, group=Wrap_RetroRainbow, parameterName=RR_Diamonds01_GradientTileRate}
  )
  n81 = TextureObjectParameter() {group=Wrap__RetroRainbow_Textures, parameterName=RR_Mask_01, sortPriority=20, texture=&T_BlacksRGB}
  n82 = TextureSample(UVs=@n22.UVs, Tex=@n81)
  n83 = ChannelMaskParameter(None=@n82.RGBA) {group=Wrap_RetroRainbow, maskChannel=Blue, parameterName=RR_Diamonds01_WarpMask, sortPriority=-26, defaultValue=<rgba 0 0 1 0>}
  n84 = ChannelMaskParameter(None=TextureSample(UVs=Add(A=@n80, B=@n83.R), Tex=@n81).RGBA) {group=Wrap_RetroRainbow, maskChannel=Alpha, parameterName=RR_Diamonds_GradientMask_01, sortPriority=-26, defaultValue=<rgba 0 0 0 1>}
  n85 = ChannelMaskParameter(None=@n82.RGBA) {group=Wrap_RetroRainbow, parameterName=RR_Diamonds01_ChannelMask, sortPriority=-26}
  n86 = Multiply(A=@n84.R, B=@n85.R)
  Saturate(None=Multiply(A=@n86, B=Add(A=@n86, B=@n85.R)))
  n87 = Time()
  n88 = ScalarParameter() {defaultValue=0.5, group=Wrap_RetroRainbow, parameterName=RR_Diamonds02_GradientSpeed, sortPriority=-26}
  n89 = Multiply(
   A=Add(A=Multiply(A=@n87, B=@n88), B=@n22.UVs)
   B=ScalarParameter() {defaultValue=1, group=Wrap_RetroRainbow, parameterName=RR_Diamonds02_GradientTileRate}
  )
  n90 = ChannelMaskParameter(None=@n82.RGBA) {group=Wrap_RetroRainbow, maskChannel=Blue, parameterName=RR_Diamonds02_WarpMask, sortPriority=-26, defaultValue=<rgba 0 0 1 0>}
  n91 = ChannelMaskParameter(None=TextureSample(UVs=Add(A=@n89, B=@n90.R), Tex=@n81).RGBA) {group=Wrap_RetroRainbow, maskChannel=Alpha, parameterName=RR_Diamonds_GradientMask_02, sortPriority=-26, defaultValue=<rgba 0 0 0 1>}
  n92 = ChannelMaskParameter(None=@n82.RGBA) {group=Wrap_RetroRainbow, maskChannel=Green, parameterName=RR_Diamonds02_ChannelMask, sortPriority=-26, defaultValue=<rgba 0 1 0 0>}
  n93 = Multiply(A=@n91.R, B=@n92.R)
  Saturate(None=Multiply(A=@n93, B=Add(A=@n93, B=@n92.R)))
  n94 = TextureObjectParameter() {group=Wrap__RetroRainbow_Textures, parameterName=RR_BaseColor_02, sortPriority=11, texture=&T_BlacksRGB}
  TextureSample(UVs=@n22.UVs, Tex=@n94)
 ]}
