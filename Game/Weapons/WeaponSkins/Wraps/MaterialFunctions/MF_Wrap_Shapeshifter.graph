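# Material function for the animated Shapeshifter cosmetic wrap on weapons, vehicles and mechs, selected through IsState.
# Pre-skinned bounds and normals drive projected UVs, which scrolling mask textures distort.
# Two warp layers form an emissive that darkens base color and is boosted by a Blueprint-timestamped reactive decay curve.
# The result, including Fresnel specular, blends over the input material through customization masks, behind the Wrap_Shapeshifter static switch.
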MaterialFunction /Game/Weapons/WeaponSkins/Wraps/MaterialFunctions/MF_Wrap_Shapeshifter
graph {nodes [
  n1 = Sine(None=Multiply(A=Time()) {constB=0.1}) {period=0.5}
  Clamp(
   None=Divide(A=Add(A=@n1))
   Min=ScalarParameter() {defaultValue=0.05, group=Wrap_ShapeShifter, parameterName=SS_Warp2_PulseMin}
   Max=ScalarParameter() {defaultValue=1, group=Wrap_ShapeShifter, parameterName=SS_Warp2_PulseMax}
  )
  Constant()
  FunctionInput(Preview=Constant3Vector() {constant=<rgba 0 1 0 0>}) {bUsePreviewValueAsDefault=true, inputName=Foward Axis, sortPriority=24}
  n2 = TextureObjectParameter() {group=Wrap__ShapeShifter_Textures, parameterName=SS_Mask02Texture, texture=&T_BlacksRGB}
  n3 = FunctionInput(Preview=Constant()) {bUsePreviewValueAsDefault=true, inputName=IsState, inputType=FunctionInput_Vector4, sortPriority=4}
  n4 = MaterialFunctionCall(Color=@n3) {materialFunction=&SplitComponents}
  n5 = LinearInterpolate(
   A=LinearInterpolate(A=Constant3Vector() {constant=<rgba 0 0 -0.25 1>}, B=Constant3Vector() {constant=<rgba 0 0 0 1>}, Alpha=@n4.R)
   B=Constant3Vector() {constant=<rgba 0.25 0.25 0 1>}
   Alpha=@n4.B
  )
  n6 = VectorParameter() {group=Wrap_ShapeShifter, parameterName=SS_Warp2_Speed_Weapon, defaultValue=<rgba 0 0 0 1>}
  n7 = MaterialFunctionCall(Color=@n3) {materialFunction=&SplitComponents}
  n8 = LinearInterpolate(
   A=VectorParameter() {group=Wrap_ShapeShifter, parameterName=SS_Warp2_Speed_Vehicle, defaultValue=<rgba 0 0 0 1>}.RGB
   B=VectorParameter() {group=Wrap_ShapeShifter, parameterName=SS_Warp2_Speed_VehicleSide, defaultValue=<rgba 0 0 0 1>}.RGB
   Alpha=@n7.G
  )
  n9 = Multiply(
   A=Time()
   B=LinearInterpolate(
    A=LinearInterpolate(A=@n6.RGB, B=@n8, Alpha=@n7.R)
    B=VectorParameter() {group=Wrap_ShapeShifter, parameterName=SS_Warp2_Speed_Mech, defaultValue=<rgba 0 0 0 1>}.RGB
    Alpha=@n7.B
   )
  )
  n10 = VertexInterpolator(VS=PreSkinnedPosition())
  n11 = PreSkinnedLocalBounds()
  n12 = FunctionInput(Preview=Divide(A=Subtract(A=@n10.PS, B=@n11.Min), B=Subtract(A=@n11.Max, B=@n11.Max))) {bUsePreviewValueAsDefault=true, inputName=Calcated_Bounds, sortPriority=11}
  n13 = Multiply(A=TextureCoordinate(), B=ScalarParameter() {defaultValue=1.5, group=Wrap_ShapeShifter, parameterName=SS_Warp2_WarpTile})
  n14 = Multiply(A=Time()) {constB=-0.015}
  n15 = TextureObjectParameter() {
   group=Wrap__ShapeShifter_Textures
   parameterName=SS_Mask01Texture
   texture=/Game/Weapons/WeaponSkins/Wraps/Textures/Shapeshifter/T_Wrap_Shapeshifter_M1
  }
  n16 = Multiply(
   A=TextureSample(UVs=Add(A=@n13, B=@n14), Tex=@n15) {samplerType=SAMPLERTYPE_Masks}.B
   B=ScalarParameter() {defaultValue=0.025, group=Wrap_ShapeShifter, parameterName=SS_Warp2_WarpAmount}
  ) {constB=0.025}
  n17 = Add(A=Multiply(A=Add(A=@n12, B=@n9), B=ScalarParameter() {defaultValue=1, group=Wrap_ShapeShifter, parameterName=SS_Warp2_TileRate}), B=@n16)
  n18 = LinearInterpolate(
   A=ScalarParameter() {defaultValue=1, group=Wrap_Shapeshifter_Topside, parameterName=SS_TopSide_Blend_Contrast, sortPriority=110}
   B=ScalarParameter() {defaultValue=2, group=Wrap_Shapeshifter_Topside, parameterName=SS_TopSide_Blend_Contrast_Mech, sortPriority=110}
   Alpha=@n4.B
  )
  n19 = ScalarParameter() {group=Wrap_Shapeshifter_Topside, parameterName=SS_TopSide_Fill_alternate_Axis_Opacity 01, sortPriority=104}
  n20 = VectorParameter() {group=Wrap_Shapeshifter_Topside, parameterName=SS_TopSide_Fill_alternate_Axis_Color_Mask_01, sortPriority=105, defaultValue=<rgba 0 0 0 1>}
  n21 = AppendVector(A=@n20.RGB, B=@n20.RGB)
  n22 = ScalarParameter() {group=Wrap_Shapeshifter_Topside, parameterName=SS_TopSide_Fill_Z_Opacity 01, sortPriority=104}
  n23 = FunctionInput(Preview=VertexInterpolator(VS=PreSkinnedNormal()).PS) {bUsePreviewValueAsDefault=true, inputName=Pre-Skinned Local Normal, sortPriority=10}
  n24 = Multiply(A=Time(), B=ScalarParameter() {defaultValue=0.01, group=Wrap_ShapeShifter, parameterName=SS_Warp2_Noisespeed})
  n25 = ScalarParameter() {defaultValue=1, group=Wrap_ShapeShifter, parameterName=SS_Warp2_NoiseTileRate}
  n26 = Multiply(A=Add(A=@n12, B=@n24), B=@n25)
  n27 = ScalarParameter() {defaultValue=1, group=Wrap_ShapeShifter, parameterName=SS_X+XBlend_Contrast, sortPriority=100}
  n28 = MaterialFunctionCall(In=Abs(None=ComponentMask(None=@n23) {b=true}), Contrast=@n27) {materialFunction=&CheapContrast}
  n29 = MaterialFunctionCall(In=Abs(None=ComponentMask(None=@n23) {g=true}), Contrast=@n27) {materialFunction=&CheapContrast}
  n30 = Saturate(None=Subtract(A=@n28.Result, B=@n29.Result))
  n31 = MaterialFunctionCall(In=Abs(None=ComponentMask(None=@n23) {r=true}), Contrast=@n27) {materialFunction=&CheapContrast}
  n32 = Add(
   A=Saturate(None=Add(A=@n30, B=@n31.Result))
   B=ScalarParameter() {defaultValue=1, group=Wrap_ShapeShifter, parameterName=SS_X+XAEmissive_alternate_Axis_Masking_Weapon, sortPriority=100}
  )
  n33 = Saturate(None=Subtract(A=@n28.Result, B=@n31.Result))
  n34 = Add(
   A=Saturate(None=Add(A=@n33, B=@n29.Result))
   B=ScalarParameter() {defaultValue=1, group=Wrap_ShapeShifter, parameterName=SS_X+XAEmissive_alternate_Axis_Masking_Vehicle, sortPriority=100}
  )
  n35 = MaterialFunctionCall(Color=@n12) {materialFunction=&SplitComponents}
  n36 = Saturate(None=LinearInterpolate(A=LinearInterpolate(A=@n32, B=@n34, Alpha=@n35.R), Alpha=@n35.B))
  n37 = Multiply(
   A=@n36
   B=ComponentMask(None=MaterialFunctionCall!(
     IsCharacter
     IsStatsB
     Texture=@n2
     Rotation_Axis=@n5
     Caculated_Bounds=@n26
     Pre_Skinned_Normals=@n23
     Blend_Contrast=@n18
     Fill_alternate_Axis_Opacity=@n19
     "Fill_alternate_Axis_Color_A=Roughness"=@n21
     Is_Vehicle=@n4.R
     Is_Mech=@n4.R
     Enable_Fill_Z_Weapon
     Fill_Z_Weapon_opacity=@n22
     Fill_Z_Weapon_Color
    ) {materialFunction=&/MF_TopsSideProjection_Base}.Combined) {g=true}
  )
  n38 = Multiply(
   A=ComponentMask(None=MaterialFunctionCall(
     Texture=@n2
     Rotation_Axis=@n5
     Caculated_Bounds=@n17
     Pre_Skinned_Normals=@n23
     Blend_Contrast=@n18
     Fill_alternate_Axis_Opacity=@n19
     "Fill_alternate_Axis_Color_A=Roughness"=@n21
     Is_Vehicle=@n4.R
     Is_Mech=@n4.R
     Fill_Z_Weapon_opacity=@n22
    ) {materialFunction=&MF_TopsSideProjection_Base}.Combined) {b=true}
   B=@n37
  )
  n39 = Multiply(A=@n38, B=Add(A=@n38, B=@n37))
  n40 = MaterialFunctionCall(Color=@n3) {materialFunction=&SplitComponents}
  n41 = LinearInterpolate(
   A=VectorParameter() {group=Wrap_ShapeShifter, parameterName=SS_Warp1_Speed_Weapon, defaultValue=<rgba 0 0 0 1>}.RGB
   B=LinearInterpolate(
    A=VectorParameter() {group=Wrap_ShapeShifter, parameterName=SS_Warp1_Speed_Vehicle, defaultValue=<rgba 0 0 0 1>}.RGB
    B=VectorParameter() {group=Wrap_ShapeShifter, parameterName=SS_Warp1_Speed_VehicleSide, defaultValue=<rgba 0 0 0 1>}.RGB
    Alpha=@n40.G
   )
   Alpha=@n40.R
  )
  n42 = VectorParameter() {group=Wrap_ShapeShifter, parameterName=SS_Warp1_Speed_Mech, defaultValue=<rgba 0 0 0 1>}
  n43 = ComponentMask(None=Multiply(A=Time(), B=LinearInterpolate(A=@n41, B=@n42.RGB, Alpha=@n40.B))) {g=true, r=true}
  n44 = ScalarParameter() {defaultValue=1, group=Wrap_ShapeShifter, parameterName=SS_BaseColor_Warp_TileRate}
  n45 = MaterialFunctionCall(Color=@n3) {materialFunction=&SplitComponents}
  n46 = VectorParameter() {group=Wrap_ShapeShifter, parameterName=SS_Offset_Weapon, defaultValue=<rgba 0 0 0 1>}
  n47 = VectorParameter() {desc=must activate TS_Vehicle_Side_01 UVs, group=Wrap_ShapeShifter, parameterName=SS_Offset_Vehicle_Side, defaultValue=<rgba 0 0 0 1>}
  n48 = MaterialFunctionCall(Color=@n3) {materialFunction=&SplitComponents}
  n49 = LinearInterpolate(A=VectorParameter() {group=Wrap_ShapeShifter, parameterName=SS_Offset_Vehicle, defaultValue=<rgba 0 0 0 1>}.RGB, B=@n47.RGB, Alpha=@n48.G)
  n50 = ScalarParameter() {defaultValue=1, group=Wrap_ShapeShifter, parameterName=SS_TileRate_Weapon}
  n51 = LinearInterpolate(
   A=ScalarParameter() {defaultValue=1, group=Wrap_ShapeShifter, parameterName=SS_TileRate_Vehicle}
   B=ScalarParameter() {defaultValue=1, group=Wrap_ShapeShifter, parameterName=SS_TileRate_VehicleSide}
   Alpha=@n48.G
  )
  n52 = MaterialFunctionCall!(
   For_Safe_Zones
   Normal_Map
   Caculated_Bounds=@n12
   Pre_Skinned_Normals=@n23
   Is Vehicle=@n45.R
   Is VehicleSide=@n45.R
   Is Mech=@n45.R
   Vehicle_Rotation=ScalarParameter() {group=Wrap_ShapeShifter, parameterName=SS_Rotation_Vehicle}
   UV_Rotation Bypass
   Mech_Rotation=ScalarParameter() {defaultValue=0.25, group=Wrap_ShapeShifter, parameterName=SS_Rotation_Mech}
   Uvs_Offset=LinearInterpolate(
    A=LinearInterpolate(A=@n46.RGB, B=@n49, Alpha=@n48.R)
    B=VectorParameter() {group=Wrap_ShapeShifter, parameterName=SS_Offset_Mech, defaultValue=<rgba 0 0 0 1>}.RGB
    Alpha=@n48.B
   )
   Offset-after rotation
   Uvs_TileRate=LinearInterpolate(
    A=LinearInterpolate(A=@n50, B=@n51, Alpha=@n48.R)
    B=ScalarParameter() {defaultValue=1, group=Wrap_ShapeShifter, parameterName=SS_TileRate_Mech}
    Alpha=@n48.B
   )
  ) {materialFunction=&/MF_Projection_Side_Normal}
  n53 = TextureSample(UVs=Add(A=Multiply(A=@n43, B=@n44), B=@n52.UVs), Tex=@n15) {samplerType=SAMPLERTYPE_LinearGrayscale}
  n54 = ScalarParameter() {defaultValue=0.1, group=Wrap_ShapeShifter, parameterName=SS_BaseColor_Warp_Amount}
  n55 = Saturate(None=TextureSample(UVs=@n52.UVs, Tex=@n15) {samplerType=SAMPLERTYPE_LinearGrayscale}.A)
  n56 = Add(A=@n52.UVs, B=LinearInterpolate(B=Multiply(A=@n53.B, B=@n54), Alpha=@n55))
  n57 = TextureSample(UVs=@n56, Tex=@n2)
  n58 = Multiply(A=Multiply(A=ScalarParameter() {group=Wrap_ShapeShifter, parameterName=SS_Warp2_EmissiveIntensity}, B=Saturate(None=@n39)), B=@n57.A)
  n59 = TextureSample(UVs=@n56, Tex=@n15) {samplerType=SAMPLERTYPE_Masks}
  n60 = Add(
   A=@n52.UVs
   B=Multiply(
    A=TextureSample(UVs=@n56, Tex=@n15) {samplerType=SAMPLERTYPE_Masks}.G
    B=ScalarParameter() {defaultValue=0.2, group=Wrap_ShapeShifter, parameterName=SS_Warp1_WarpTimeAmount}
   )
  )
  n61 = ScalarParameter() {defaultValue=1, group=Wrap_ShapeShifter, parameterName=SS_Warp1_TileRate}
  n62 = TextureSample(UVs=Add(A=Multiply(A=@n60, B=@n61), B=@n43), Tex=@n15) {samplerType=SAMPLERTYPE_LinearGrayscale}
  n63 = ScalarParameter() {defaultValue=1, group=Wrap_ShapeShifter, parameterName=SS_Warp1_Contrast}
  n64 = Multiply(
   A=@n36
   B=Multiply(
    A=ScalarParameter() {group=Wrap_ShapeShifter, parameterName=SS_Warp1_Amount}
    B=MaterialFunctionCall(In=@n62.B, Contrast=@n63) {materialFunction=&CheapContrast}.Result
   )
  )
  n65 = Multiply(A=@n59.R, B=@n64)
  n66 = Multiply(A=@n65, B=Add(A=@n65, B=@n64))
  n67 = Multiply(
   A=@n59.R
   B=Multiply(A=Saturate(None=@n66), B=ScalarParameter() {defaultValue=1, group=Wrap_ShapeShifter, parameterName=SS_Warp1_EmissiveIntensity})
  )
  n68 = Time()
  n69 = ScalarParameter() {defaultValue=-9999999, group=Reactive, parameterName=TimeFrom-BP----------------, sortPriority=-600}
  n70 = Saturate(None=Divide(A=Subtract(A=@n68, B=@n69), B=ScalarParameter() {defaultValue=1, group=Reactive, parameterName=SS_AnimationDuration, sortPriority=33}))
  n71 = ScalarParameter() {defaultValue=0.5, group=Reactive, parameterName=SS_Curve_Amount, sortPriority=34}
  n72 = ScalarParameter() {group=Reactive, parameterName=SS_TestReactive, sortPriority=300}
  n73 = LinearInterpolate(A=OneMinus(None=Power(Base=@n70, Exp=@n71)) {desc=inverts curve from 0-1 to 1-0}, B=@n72, Alpha=@n72)
  n74 = LinearInterpolate(A=Add(A=@n58, B=@n67), B=Add(A=@n58, B=@n67), Alpha=@n73)
  n75 = ScalarParameter() {defaultValue=4, group=Reactive, parameterName=SS_Emissive_AmountBlend}
  n76 = TextureObjectParameter() {group=Wrap__ShapeShifter_Textures, parameterName="SS_EmissiveTexture A=Specular", texture=&T_BlacksRGB}
  n77 = TextureSample(UVs=@n56, Tex=@n76) {samplerType=SAMPLERTYPE_Masks}
  n78 = ScalarParameter() {defaultValue=1, group=Reactive, parameterName=SS_Emissive_AmountReactive}
  n79 = LinearInterpolate(A=@n74, B=Add(A=Multiply(A=@n74, B=@n75), B=Multiply(A=@n77.RGB, B=@n78)), Alpha=@n73)
  n80 = Add(A=Add(A=@n57.R, B=@n59.R), B=@n79)
  n81 = Saturate(None=Multiply(
    A=ScalarParameter() {defaultValue=1, group=Wrap_ShapeShifter, parameterName=SS_Emissive01-DarkenBase-------, sortPriority=-20}
    B=LinearInterpolate(A=@n79, B=@n80, Alpha=@n73)
   ))
  n82 = TextureObjectParameter() {group=Wrap__ShapeShifter_Textures, parameterName="SS_BaseColor A=Roughness", texture=&T_BlacksRGB}
  n83 = TextureSample(UVs=@n56, Tex=@n82) {samplerType=SAMPLERTYPE_Masks}
  n84 = GetMaterialAttributes(None=FunctionInput() {inputName=Material, inputType=FunctionInput_MaterialAttributes, sortPriority=-20}) {attributeGetTypes [#1]}
  n85 = Fresnel(
   ExponentIn=ScalarParameter() {defaultValue=2.5, group=Wrap_ShapeShifter, parameterName=SS_SpecularFresnel, sortPriority=1}
   Normal=Transform(None=@n84.Normal)
  )
  n86 = ScalarParameter() {group=Wrap_ShapeShifter, parameterName=SS_SpecularFresnel_Contrast, sortPriority=1}
  n87 = ScalarParameter() {defaultValue=0.5, group=Wrap_ShapeShifter, parameterName=SS_Specular, sortPriority=1}
  n88 = ScalarParameter() {defaultValue=0.5, group=Wrap_ShapeShifter, parameterName=SS_Specular_High, sortPriority=1}
  n89 = SetMaterialAttributes!(
   MaterialAttributes=@n84.MaterialAttributes
   BaseColor=Multiply(A=@n83.RGB, B=OneMinus(None=@n81))
   Metallic=ScalarParameter() {group=Wrap_ShapeShifter, parameterName=SS_Metallic, sortPriority=1}
   Specular=Add(
    A=MaterialFunctionCall(In=@n85, Contrast=@n86) {materialFunction=&CheapContrast}.Result
    B=LinearInterpolate(A=@n87, B=@n88, Alpha=@n77.A)
   )
   Roughness=LinearInterpolate(
    A=ScalarParameter() {defaultValue=0.15, group=Wrap_ShapeShifter, parameterName=SS_Roughness, sortPriority=1}
    B=ScalarParameter() {defaultValue=1, group=Wrap_ShapeShifter, parameterName=SS_Roughness_High, sortPriority=1}
    Alpha=@n83.A
   )
   EmissiveColor=@n79
  ) {attributeSetTypes [
    #2
    #3
    #4
    #5
    #6
   ]}
  n90 = FunctionInput() {
   bUsePreviewValueAsDefault=true
   desc=base customization
   inputName="Customization Mask (_CM B Top left)"
   inputType=FunctionInput_Scalar
   sortPriority=49
  }
  n91 = FunctionInput() {
   bUsePreviewValueAsDefault=true
   desc=secondary
   inputName="Customization Mask 2 (_CM B Top Right)"
   inputType=FunctionInput_Scalar
   sortPriority=50
  }
  n92 = NamedRerouteUsage()
  n93 = NamedRerouteUsage()
  n94 = LinearInterpolate(
   A=Saturate(None=MaterialFunctionCall!("Customization Mask (_CM B Top left)"=@n90, "Customization Mask 2 (_CM B Top Right)"=@n91) {materialFunction=&/MF_Customization_AllMasking}.Result)
   B=@n90
   Alpha=Saturate(None=Subtract(A=@n92, B=@n93))
  )
  FunctionOutput(None=StaticSwitchParameter(True=BlendMaterialAttributes(A=@n84.MaterialAttributes, B=@n89, Alpha=@n94) {vertexAttributeBlendType=UseA}, False=@n84.MaterialAttributes) {group=Customization_Wrap, parameterName=Wrap_Shapeshifter}) {outputName=Result_1}
  n95 = ScalarParameter() {defaultValue=1, group=Reactive, parameterName=SS_Emissive01-ReactiveDarkenBase}
  LinearInterpolate(A=@n83.RGB, B=LinearInterpolate(A=@n83.RGB, B=Multiply(A=@n83.RGB, B=@n95), Alpha=@n55), Alpha=@n73)
  n96 = Time()
  Multiply(A=Add(A=Multiply(A=@n96) {constB=-0.015}, B=@n52.UVs)) {constB=0.75}
  n97 = Constant() {r=0.75}
  Multiply(
   A=@n79
   B=MaterialFunctionCall(Value_Morning=@n97, Value_Day=Constant() {r=1}, Value_Evening=@n97, Value_Night=Constant() {r=0.5}) {materialFunction=&Scalar_Time_of_Day}.Result
  )
  Multiply(A=@n77.RGB, B=@n67)
  NamedRerouteDeclaration(None=MaterialFunctionCall(Float4=@n3) {materialFunction=&BreakOutFloat4Components}.A) {name=IsCharacter, nodeColor=<rgba 0.909 1 0 1>}
  NamedRerouteDeclaration(None=FunctionInput() {bUsePreviewValueAsDefault=true, inputName=isCharAccent, inputType=FunctionInput_Scalar, sortPriority=78}) {name=isCharAccent, nodeColor=<rgba 0 0.9253 1 1>}
  n98 = Add(A=ComponentMask(None=@n9) {g=true, r=true}, B=@n52.UVs)
  TextureSample(UVs=Multiply(A=Add(A=@n98, B=@n16), B=ScalarParameter() {defaultValue=1, group=Wrap_ShapeShifter, parameterName=SS_Warp2_TileRate}), Tex=@n2)
  TextureSample(UVs=Multiply(A=Add(A=@n24, B=@n52.UVs), B=@n25), Tex=@n2)
 ]}
