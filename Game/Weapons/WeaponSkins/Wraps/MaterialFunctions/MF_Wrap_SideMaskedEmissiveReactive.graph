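# Weapon wrap material function adding a side-projected skin with reactive emissive, blended over the input material by customization masks behind the Wrap_AncientGladiator switch.
# MF_Projection_Side_Normal supplies UVs and normals, with IsState selecting weapon, vehicle or mech settings.
# Masks drive metal, specular and scrolling emissive, which pulses, scales by time of day and darkens base color.
# A Blueprint timestamp drives a decaying blend toward reactive emissive.
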MaterialFunction /Game/Weapons/WeaponSkins/Wraps/MaterialFunctions/MF_Wrap_SideMaskedEmissiveReactive
graph {nodes [
  n1 = GetMaterialAttributes(None=FunctionInput() {inputName=Material, inputType=FunctionInput_MaterialAttributes, sortPriority=-20}) {attributeGetTypes [#1]}
  n2 = Fresnel(
   ExponentIn=ScalarParameter() {defaultValue=2, group=Wrap_SideMaskedEmissiveReactive, parameterName=SMER_SpecularFresnel, sortPriority=1}
   Normal=Transform(None=@n1.Normal)
  )
  n3 = ScalarParameter() {group=Wrap_SideMaskedEmissiveReactive, parameterName=SMER_SpecularFresnelContrast, sortPriority=1}
  n4 = TextureObjectParameter() {
   group=Wrap__SideMaskedEmissiveReactive_Textures
   parameterName="SMER_MaskMap_01 - For Metal, Specular, Warp, Emissive Mask"
   texture=&T_BlacksRGB
  }
  n5 = TextureObjectParameter() {
   group=Wrap__SideMaskedEmissiveReactive_Textures
   parameterName=SMER_Normal
   samplerType=SAMPLERTYPE_Normal
   texture=&FlatNormalBlue
  }
  n6 = FunctionInput(Preview=Constant()) {bUsePreviewValueAsDefault=true, inputName=IsState, inputType=FunctionInput_Vector4}
  n7 = MaterialFunctionCall(Color=@n6) {materialFunction=&SplitComponents}
  n8 = VectorParameter() {group=Wrap_SideMaskedEmissiveReactive_UVs, parameterName=SMER_Offset_Weapon, defaultValue=<rgba 0 0 0 1>}
  n9 = VectorParameter() {group=Wrap_SideMaskedEmissiveReactive_UVs, parameterName=SMER_Offset_Vehicle, defaultValue=<rgba 0 0 0 1>}
  n10 = VectorParameter() {
   desc=must activate TS_Vehicle_Side_01 UVs
   group=Wrap_SideMaskedEmissiveReactive_UVs
   parameterName=SMER_Offset_Vehicle_Side
   defaultValue=<rgba 0 0 0 1>
  }
  n11 = MaterialFunctionCall(Color=@n6) {materialFunction=&SplitComponents}
  n12 = VectorParameter() {group=Wrap_SideMaskedEmissiveReactive_UVs, parameterName=SMER_Offset_Mech -----, defaultValue=<rgba 0 0 0 1>}
  n13 = LinearInterpolate(A=LinearInterpolate(A=@n8.RGB, B=LinearInterpolate(A=@n9.RGB, B=@n10.RGB, Alpha=@n11.G), Alpha=@n11.R), B=@n12.RGB, Alpha=@n11.B)
  n14 = ScalarParameter() {defaultValue=1, group=Wrap_SideMaskedEmissiveReactive_UVs, parameterName=SMER_TileRate_Weapon}
  n15 = LinearInterpolate(
   A=ScalarParameter() {defaultValue=1, group=Wrap_SideMaskedEmissiveReactive_UVs, parameterName=SMER_TileRate_Vehicle}
   B=ScalarParameter() {defaultValue=1, group=Wrap_SideMaskedEmissiveReactive_UVs, parameterName=SMER_TileRate_VehicleSide}
   Alpha=@n11.G
  )
  n16 = FunctionInput(Preview=VertexInterpolator(VS=PreSkinnedNormal()).PS) {bUsePreviewValueAsDefault=true, inputName=Pre-Skinned Local Normal, sortPriority=10}
  n17 = VertexInterpolator(VS=PreSkinnedPosition())
  n18 = PreSkinnedLocalBounds()
  n19 = FunctionInput(Preview=Divide(A=Subtract(A=@n17.PS, B=@n18.Min), B=Subtract(A=@n18.Max, B=@n18.Max))) {bUsePreviewValueAsDefault=true, inputName=Calcated_Bounds, sortPriority=11}
  n20 = MaterialFunctionCall!(
   For_Safe_Zones
   Normal_Map=@n5
   Caculated_Bounds=@n19
   Pre_Skinned_Normals=@n16
   Is Vehicle=@n7.R
   Is VehicleSide=@n7.R
   Is Mech=@n7.R
   Vehicle_Rotation=ScalarParameter() {group=Wrap_SideMaskedEmissiveReactive_UVs, parameterName=SMER_Rotation_Vehicle}
   UV_Rotation Bypass
   Mech_Rotation=ScalarParameter() {defaultValue=0.25, group=Wrap_SideMaskedEmissiveReactive_UVs, parameterName=SMER_Rotation_Mech -----}
   Uvs_Offset=@n13
   Offset-after rotation
   Uvs_TileRate=LinearInterpolate(
    A=LinearInterpolate(A=@n14, B=@n15, Alpha=@n11.R)
    B=ScalarParameter() {defaultValue=1, group=Wrap_SideMaskedEmissiveReactive_UVs, parameterName=SMER_TileRate_Mech -----}
    Alpha=@n11.B
   )
  ) {materialFunction=&/MF_Projection_Side_Normal}
  n21 = TextureSample(UVs=@n20.UVs, Tex=@n4) {samplerType=SAMPLERTYPE_Masks}
  n22 = LinearInterpolate(
   A=ScalarParameter() {defaultValue=0.5, group=Wrap_SideMaskedEmissiveReactive, parameterName=SMER_Specular_Min, sortPriority=1}
   B=ScalarParameter() {defaultValue=0.5, group=Wrap_SideMaskedEmissiveReactive, parameterName=SMER_Specular_Max, sortPriority=1}
   Alpha=ChannelMaskParameter(None=@n21.RGBA) {group=Wrap_SideMaskedEmissiveReactive, parameterName=SMER_Specular_Channel, sortPriority=1}.R
  )
  Add(
   A=LinearInterpolate(
    B=MaterialFunctionCall(In=@n2, Contrast=@n3) {materialFunction=&CheapContrast}.Result
    Alpha=ScalarParameter() {group=Wrap_SideMaskedEmissiveReactive, parameterName="SMER_SpecularFresnelControl 1=Use Fresnel", sortPriority=1}
   )
   B=@n22
  )
  n23 = VectorParameter() {group=Wrap_SideMaskedEmissiveReactive_Emissive, parameterName=SMER_Emissive1_WarpSpeed_Weapon, defaultValue=<rgba 0 0 0 1>}
  n24 = VectorParameter() {group=Wrap_SideMaskedEmissiveReactive_Emissive, parameterName=SMER_Emissive1_WarpSpeed_Vehicle, defaultValue=<rgba 0 0 0 1>}
  n25 = VectorParameter() {group=Wrap_SideMaskedEmissiveReactive_Emissive, parameterName=SMER_Emissive1_WarpSpeed_VehicleSide, defaultValue=<rgba 0 0 0 1>}
  n26 = MaterialFunctionCall(Color=@n6) {materialFunction=&SplitComponents}
  n27 = LinearInterpolate(A=@n23.RGB, B=LinearInterpolate(A=@n24.RGB, B=@n25.RGB, Alpha=@n26.G), Alpha=@n26.R)
  n28 = VectorParameter() {group=Wrap_SideMaskedEmissiveReactive_Emissive, parameterName=SMER_Emissive1_WarpSpeed_Mech, defaultValue=<rgba 0 0 0 1>}
  ComponentMask(None=Multiply(A=Time(), B=LinearInterpolate(A=@n27, B=@n28.RGB, Alpha=@n26.B))) {g=true, r=true}
  Constant3Vector() {constant=<rgba 0 0 1 0>}
  Constant()
  n29 = ScalarParameter() {defaultValue=1, group=Wrap_SideMaskedEmissiveReactive_Emissive, parameterName=SMER_Emissive_DarkenBaseColor, sortPriority=-20}
  n30 = TextureObjectParameter() {
   group=Wrap__SideMaskedEmissiveReactive_Textures
   parameterName=SMER_MaskMap_02 - Emissive mask and noises
   texture=&T_BlacksRGB
  }
  n31 = TextureSample(UVs=@n20.UVs, Tex=@n30) {samplerType=SAMPLERTYPE_Masks}
  n32 = ChannelMaskParameter(None=@n31.RGBA) {
   group=Wrap_SideMaskedEmissiveReactive_Emissive
   maskChannel=Green
   parameterName=SMER__EmissiveMaskB_Channel
   sortPriority=11
   defaultValue=<rgba 0 1 0 0>
  }
  n33 = Multiply(A=VectorParameter() {group=Wrap_SideMaskedEmissiveReactive_Emissive, parameterName=SMER__EmissiveBColor, sortPriority=0}.RGB, B=@n32.R)
  n34 = ScalarParameter() {defaultValue=1, group=Wrap_SideMaskedEmissiveReactive_Reactive, parameterName=SMER_Reactive_ReactiveAmount}
  n35 = Time()
  n36 = ScalarParameter() {defaultValue=-9999999, group=Wrap_SideMaskedEmissiveReactive_Reactive, parameterName=TimeFrom-BP----------------, sortPriority=-600}
  n37 = Saturate(None=Divide(
    A=Subtract(A=@n35, B=@n36)
    B=ScalarParameter() {defaultValue=1.5, group=Wrap_SideMaskedEmissiveReactive_Reactive, parameterName=SMER_AnimationDuration, sortPriority=33}
   ))
  n38 = ScalarParameter() {defaultValue=0.85, group=Wrap_SideMaskedEmissiveReactive_Reactive, parameterName=SMER_Curve_Amount, sortPriority=34}
  n39 = ScalarParameter() {group=Wrap_SideMaskedEmissiveReactive_Reactive, parameterName=SMER_TestReactive, sortPriority=300}
  n40 = LinearInterpolate(A=OneMinus(None=Power(Base=@n37, Exp=@n38)) {desc=inverts curve from 0-1 to 1-0}, B=@n39, Alpha=@n39)
  n41 = LinearInterpolate(A=@n33, B=Multiply(A=@n33, B=@n34), Alpha=@n40)
  n42 = Time()
  n43 = ScalarParameter() {defaultValue=0.1, group=Wrap_SideMaskedEmissiveReactive_Emissive, parameterName=SMER_Emissive_PulseTime}
  n44 = Add(A=Sine(None=Multiply(A=@n42, B=@n43) {constB=0.1}) {period=0.5})
  n45 = Multiply(
   A=Clamp(
    None=Divide(A=@n44)
    Min=ScalarParameter() {defaultValue=0.05, group=Wrap_SideMaskedEmissiveReactive_Emissive, parameterName=SMER_Emissive_PulseMin}
    Max=ScalarParameter() {defaultValue=1, group=Wrap_SideMaskedEmissiveReactive_Emissive, parameterName=SMER_Emissive_PulseMax}
   )
   B=ScalarParameter() {defaultValue=1, group=Wrap_SideMaskedEmissiveReactive_Emissive, parameterName=SMER_Emissive_PulseIntensity}
  )
  n46 = ChannelMaskParameter(None=@n31.RGBA) {group=Wrap_SideMaskedEmissiveReactive_Reactive, maskChannel=Blue, parameterName=SMER_Reactive_MaskChannel -----, defaultValue=<rgba 0 0 1 0>}
  n47 = ScalarParameter() {group=Wrap_SideMaskedEmissiveReactive_Reactive, parameterName="SMER_Reactive_MaskChannelOverride 1=Use New Mask for Reactive"}
  n48 = LinearInterpolate(
   A=ChannelMaskParameter(None=@n31.RGBA) {group=Wrap_SideMaskedEmissiveReactive_Emissive, parameterName=SMER__EmissiveMaskA_Channel, sortPriority=11}.R
   B=@n46.R
   Alpha=LinearInterpolate(B=@n47, Alpha=@n40)
  )
  n49 = LinearInterpolate(
   A=@n48
   B=Multiply(A=@n48, B=@n45)
   Alpha=ScalarParameter() {group=Wrap_SideMaskedEmissiveReactive_Emissive, parameterName="SMER_Emissive_PulseControl 1=Use Pulse -----"}
  )
  n50 = Add(
   A=Multiply(
    A=Time()
    B=ScalarParameter() {defaultValue=0.01, group=Wrap_SideMaskedEmissiveReactive_Emissive, parameterName=SMER_Emissive2_WarpGradientspeed}
   )
   B=@n20.UVs
  )
  n51 = ScalarParameter() {defaultValue=1, group=Wrap_SideMaskedEmissiveReactive_Emissive, parameterName=SMER_Emissive2_WarpGradientTileRate}
  n52 = ChannelMaskParameter(None=TextureSample(UVs=Multiply(A=@n50, B=@n51), Tex=@n30).RGBA) {
   group=Wrap_SideMaskedEmissiveReactive_Emissive
   maskChannel=Blue
   parameterName=SMER_Emissive2_WarpGradientMaskChannel
   defaultValue=<rgba 0 0 1 0>
  }
  n53 = VectorParameter() {group=Wrap_SideMaskedEmissiveReactive_Emissive, parameterName=SMER_Emissive2_WarpSpeed_Weapon, defaultValue=<rgba 0 0 0 1>}
  n54 = VectorParameter() {group=Wrap_SideMaskedEmissiveReactive_Emissive, parameterName=SMER_Emissive2_WarpSpeed_Vehicle, defaultValue=<rgba 0 0 0 1>}
  n55 = VectorParameter() {group=Wrap_SideMaskedEmissiveReactive_Emissive, parameterName=SMER_Emissive2_WarpSpeed_VehicleSide, defaultValue=<rgba 0 0 0 1>}
  n56 = MaterialFunctionCall(Color=@n6) {materialFunction=&SplitComponents}
  n57 = LinearInterpolate(A=@n53.RGB, B=LinearInterpolate(A=@n54.RGB, B=@n55.RGB, Alpha=@n56.G), Alpha=@n56.R)
  n58 = VectorParameter() {group=Wrap_SideMaskedEmissiveReactive_Emissive, parameterName=SMER_Emissive2_WarpSpeed_Mech, defaultValue=<rgba 0 0 0 1>}
  n59 = ComponentMask(None=Multiply(A=Time(), B=LinearInterpolate(A=@n57, B=@n58.RGB, Alpha=@n56.B))) {g=true, r=true}
  n60 = Multiply(
   A=TextureCoordinate()
   B=ScalarParameter() {defaultValue=1, group=Wrap_SideMaskedEmissiveReactive_Emissive, parameterName=SMER_Emissive2_WarpTileRate}
  )
  n61 = ChannelMaskParameter(None=TextureSample(UVs=Add(A=@n59, B=@n60), Tex=@n30).RGBA) {group=Wrap_SideMaskedEmissiveReactive_Emissive, maskChannel=Green, parameterName=SMER_Emissive2_MaskChannel, defaultValue=<rgba 0 0 1 0>}
  n62 = ScalarParameter() {defaultValue=1, group=Wrap_SideMaskedEmissiveReactive_Emissive, parameterName=SMER_X+XBlend_Contrast, sortPriority=100}
  n63 = MaterialFunctionCall(In=Abs(None=ComponentMask(None=@n16) {b=true}), Contrast=@n62) {materialFunction=&CheapContrast}
  n64 = MaterialFunctionCall(In=Abs(None=ComponentMask(None=@n16) {g=true}), Contrast=@n62) {materialFunction=&CheapContrast}
  n65 = Saturate(None=Subtract(A=@n63.Result, B=@n64.Result))
  n66 = MaterialFunctionCall(In=Abs(None=ComponentMask(None=@n16) {r=true}), Contrast=@n62) {materialFunction=&CheapContrast}
  n67 = ScalarParameter() {
   defaultValue=1
   group=Wrap_SideMaskedEmissiveReactive_Emissive
   parameterName=SMER_X+XAEmissive_alternate_Axis_Masking_Weapon
   sortPriority=100
  }
  n68 = Add(A=Saturate(None=Add(A=@n65, B=@n66.Result)), B=@n67)
  n69 = Saturate(None=Subtract(A=@n63.Result, B=@n66.Result))
  n70 = ScalarParameter() {
   defaultValue=1
   group=Wrap_SideMaskedEmissiveReactive_Emissive
   parameterName=SMER_X+XAEmissive_alternate_Axis_Masking_Vehicle
   sortPriority=100
  }
  n71 = Add(A=Saturate(None=Add(A=@n69, B=@n64.Result)), B=@n70)
  n72 = MaterialFunctionCall(Color=@n19) {materialFunction=&SplitComponents}
  n73 = Saturate(None=LinearInterpolate(A=LinearInterpolate(A=@n68, B=@n71, Alpha=@n72.R), Alpha=@n72.B))
  n74 = Multiply(A=@n73, B=@n61.R)
  n75 = Multiply(A=@n52.R, B=@n74)
  n76 = Multiply(A=@n75, B=Add(A=@n75, B=@n74))
  n77 = Multiply(
   A=Multiply(A=ScalarParameter() {group=Wrap_SideMaskedEmissiveReactive_Emissive, parameterName=SMER_Emissive2_Intensity}, B=Saturate(None=@n76))
   B=@n48
  )
  n78 = Multiply(A=@n77, B=@n49)
  n79 = Multiply(
   A=VectorParameter() {group=Wrap_SideMaskedEmissiveReactive_Emissive, parameterName=SMER__EmissiveAColor, sortPriority=0}.RGB
   B=ComponentMask(None=@n78) {r=true}
  )
  n80 = LinearInterpolate(
   A=@n48
   B=Multiply(A=@n48, B=@n45)
   Alpha=ScalarParameter() {group=Wrap_SideMaskedEmissiveReactive_Reactive, parameterName="SMER_Reactive_EmissivePulseControl 1=Use Pulse"}
  )
  n81 = Add(
   A=Multiply(
    A=Time()
    B=ScalarParameter() {defaultValue=0.01, group=Wrap_SideMaskedEmissiveReactive_Emissive, parameterName=SMER_Emissive1_WarpGradientSpeed}
   )
   B=@n20.UVs
  )
  n82 = ScalarParameter() {defaultValue=1, group=Wrap_SideMaskedEmissiveReactive_Emissive, parameterName=SMER_Emissive1_WarpGradientTileRate}
  n83 = ChannelMaskParameter(None=TextureSample(UVs=Multiply(A=@n81, B=@n82), Tex=@n4).RGBA) {
   group=Wrap_SideMaskedEmissiveReactive_Emissive
   maskChannel=Blue
   parameterName=SMER_Emissive1_WarpGradientMaskChannel
   defaultValue=<rgba 0 0 1 0>
  }
  n84 = ScalarParameter() {group=Wrap_SideMaskedEmissiveReactive_Emissive, parameterName=SMER_Emissive1_Warp_Amount}
  n85 = Multiply(
   A=TextureCoordinate()
   B=ScalarParameter() {defaultValue=1, group=Wrap_SideMaskedEmissiveReactive_Emissive, parameterName=SMER_Emissive1_WarpTileRate}
  )
  n86 = ScalarParameter() {defaultValue=0.3, group=Wrap_SideMaskedEmissiveReactive_Emissive, parameterName=SMER_Emissive1_WarpSpeed}
  n87 = ChannelMaskParameter(None=TextureSample(UVs=Add(A=@n85, B=@n86), Tex=@n4) {samplerType=SAMPLERTYPE_LinearGrayscale}.RGBA) {group=Wrap_SideMaskedEmissiveReactive_Emissive, parameterName=SMER_Emissive1_WarpMaskChannel}
  n88 = Multiply(A=@n73, B=Add(A=@n20.UVs, B=Multiply(A=@n84, B=@n87.R)))
  n89 = Multiply(A=@n83.R, B=@n88)
  n90 = Multiply(A=@n89, B=Add(A=@n89, B=@n88))
  n91 = Multiply(
   A=@n48
   B=Multiply(
    A=Saturate(None=@n90)
    B=ScalarParameter() {defaultValue=1, group=Wrap_SideMaskedEmissiveReactive_Emissive, parameterName=SMER_Emissive1_Intensity}
   )
  )
  n92 = Multiply(A=@n80, B=@n91)
  n93 = Multiply(
   A=VectorParameter() {group=Wrap_SideMaskedEmissiveReactive_Reactive, parameterName=SMER_Reactive_EmissiveColor, sortPriority=0}.RGB
   B=ComponentMask(None=@n92) {r=true}
  )
  n94 = ScalarParameter() {defaultValue=1, group=Wrap_SideMaskedEmissiveReactive_Reactive, parameterName="SMER_Reactive__EmissiveControl 0=Separate Emissive Chain"}
  n95 = LinearInterpolate(A=@n79, B=Multiply(A=LinearInterpolate(A=@n93, B=@n79, Alpha=@n94), B=@n34), Alpha=@n40)
  n96 = ScalarParameter() {defaultValue=0.5, group=Wrap_SideMaskedEmissiveReactive_Emissive, parameterName=SMER_Emissive_TOD_DawnDusk-----}
  n97 = Multiply(
   A=LinearInterpolate(A=Add(A=@n41, B=@n95), B=@n95, Alpha=LinearInterpolate(B=@n47, Alpha=@n40))
   B=MaterialFunctionCall(
    Value_Morning=@n96
    Value_Day=ScalarParameter() {defaultValue=1, group=Wrap_SideMaskedEmissiveReactive_Emissive, parameterName=SMER_Emissive_TOD_Day}
    Value_Evening=@n96
    Value_Night=ScalarParameter() {defaultValue=0.25, group=Wrap_SideMaskedEmissiveReactive_Emissive, parameterName=SMER_Emissive_TOD_Night}
   ) {materialFunction=&Scalar_Time_of_Day}.Result
  )
  n98 = Saturate(None=Multiply(A=@n29, B=@n97))
  n99 = TextureObjectParameter() {
   group=Wrap__SideMaskedEmissiveReactive_Textures
   parameterName="SMER_BaseColor A=Roughness"
   texture=/Game/Weapons/WeaponSkins/Wraps/Textures/UVCheaker
  }
  n100 = TextureSample(UVs=@n20.UVs, Tex=@n99) {samplerType=SAMPLERTYPE_Masks}
  n101 = LinearInterpolate(
   A=@n1.Normal
   B=Add(A=@n1.Normal, B=Clamp(None=@n20.Raw_Normal))
   Alpha=ScalarParameter() {defaultValue=1, group=Wrap_SideMaskedEmissiveReactive, parameterName=SMER_Normal_Intensty, sortPriority=1}
  )
  n102 = ShadingPathSwitch(Default=@n101, Deferred=@n101, Mobile=@n1.Normal) {desc=Migrated from FeatureLevelSwitch}
  n103 = ChannelMaskParameter(None=@n21.RGBA) {group=Wrap_SideMaskedEmissiveReactive, parameterName=SMER_Metallic_Channel, sortPriority=1}
  n104 = SetMaterialAttributes!(
   MaterialAttributes=@n1.MaterialAttributes
   BaseColor=Multiply(A=@n100.RGB, B=OneMinus(None=@n98))
   Metallic=@n103.R
   Specular=@n22
   Roughness=LinearInterpolate(
    A=ScalarParameter() {defaultValue=0.15, group=Wrap_SideMaskedEmissiveReactive, parameterName=SMER_Roughness, sortPriority=1}
    B=ScalarParameter() {defaultValue=1, group=Wrap_SideMaskedEmissiveReactive, parameterName=SMER_Roughness_High, sortPriority=1}
    Alpha=@n100.A
   )
   EmissiveColor=@n97
   Normal=@n102
  ) {attributeSetTypes [
    #2
    #3
    #4
    #5
    #6
    #1
   ]}
  n105 = FunctionInput() {
   bUsePreviewValueAsDefault=true
   desc=base customization
   inputName="Customization Mask (_CM B Top left)"
   inputType=FunctionInput_Scalar
   sortPriority=49
  }
  n106 = FunctionInput() {
   bUsePreviewValueAsDefault=true
   desc=secondary
   inputName="Customization Mask 2 (_CM B Top Right)"
   inputType=FunctionInput_Scalar
   sortPriority=50
  }
  n107 = NamedRerouteUsage()
  n108 = NamedRerouteUsage()
  n109 = LinearInterpolate(
   A=Saturate(None=MaterialFunctionCall!("Customization Mask (_CM B Top left)"=@n105, "Customization Mask 2 (_CM B Top Right)"=@n106) {materialFunction=&/MF_Customization_AllMasking}.Result)
   B=@n105
   Alpha=Saturate(None=Subtract(A=@n107, B=@n108))
  )
  FunctionOutput(None=StaticSwitchParameter(True=BlendMaterialAttributes(A=@n1.MaterialAttributes, B=@n104, Alpha=@n109) {vertexAttributeBlendType=UseA}, False=@n1.MaterialAttributes) {group=Customization_Wrap, parameterName=Wrap_AncientGladiator}) {outputName=Result_1}
  LinearInterpolate(
   A=ScalarParameter() {group=Wrap_SideMaskedEmissiveReactive, parameterName=SMER_Metallic, sortPriority=1}
   B=@n103.R
   Alpha=ScalarParameter() {group=Wrap_SideMaskedEmissiveReactive, parameterName="SMER_MetalMaskControl - 1= Use Texture Mask", sortPriority=1}
  )
  LinearInterpolate(
   A=LinearInterpolate(
    A=@n49
    B=@n78
    Alpha=ScalarParameter() {group=Wrap_SideMaskedEmissiveReactive_Emissive, parameterName="SMER_Emissive2__X+XControl 1=Use Emissive 2 -----"}
   )
   B=Multiply(A=@n49, B=@n91)
   Alpha=ScalarParameter() {group=Wrap_SideMaskedEmissiveReactive_Emissive, parameterName="SMER_Emissive1__X+XControl 1=Use Emissive 1 -----"}
  )
  LinearInterpolate(
   A=LinearInterpolate(
    A=@n80
    B=Multiply(A=@n77, B=@n80)
    Alpha=ScalarParameter() {group=Wrap_SideMaskedEmissiveReactive_Reactive, parameterName="SMER_Reactive_Emissive2X+XControl 1=Use Emissive 2"}
   )
   B=@n92
   Alpha=ScalarParameter() {group=Wrap_SideMaskedEmissiveReactive_Reactive, parameterName="SMER_Reactive_Emissive1X+XControl 1=Use Emissive 1"}
  )
  Multiply(A=@n20.Raw_Normal)
  NamedRerouteDeclaration(None=MaterialFunctionCall(Float4=@n6) {materialFunction=&BreakOutFloat4Components}.A) {name=IsCharacter, nodeColor=<rgba 0.8047 1 0 1>}
  NamedRerouteDeclaration(None=FunctionInput() {bUsePreviewValueAsDefault=true, inputName=isCharAccent, inputType=FunctionInput_Scalar, sortPriority=78}) {name=isCharAccent, nodeColor=<rgba 0 0.9253 1 1>}
  Saturate(None=@n48)
  StaticSwitchParameter(True=@n4, False=@n30) {
   defaultValue=true
   desc=Disable if you need to free up texture channel for mask 01 warp noise
   group=Wrap_SideMaskedEmissiveReactive_Emissive
   parameterName="SMER_MaskMap01 as EmissiveMask - False=MaskeMap02"
   sortPriority=-100
  }
  n110 = TextureObjectParameter() {
   group=Wrap__SideMaskedEmissiveReactive_Textures
   parameterName=SMER_Emissive - Single Texture
   texture=&T_BlacksRGB
  }
  n111 = Multiply(
   A=TextureSample(UVs=@n20.UVs, Tex=@n110) {samplerType=SAMPLERTYPE_Masks}.RGB
   B=ScalarParameter() {defaultValue=1, group=Wrap_SideMaskedEmissiveReactive_Emissive, parameterName=SMER_Emissive_TextureIntensity -----}
  )
  StaticSwitchParameter(True=@n111, False=@n93) {group=Wrap_SideMaskedEmissiveReactive_Reactive, parameterName=SMER_Reactive_UseEmissiveTexture, sortPriority=-100}
  StaticSwitchParameter(True=@n111, False=@n79) {group=Wrap_SideMaskedEmissiveReactive_Emissive, parameterName=SMER_Emissive_UseEmissiveTexture, sortPriority=-100}
  n112 = Multiply(
   A=@n20.UVs No Offsets/Tiling
   B=ScalarParameter() {defaultValue=1, group=Wrap_SideMaskedEmissiveReactive_UVs, parameterName=SMER_Warp_TileRate}
  )
  n113 = VectorParameter() {group=Wrap_SideMaskedEmissiveReactive_UVs, parameterName=SMER_Warp_Speed_Weapon, defaultValue=<rgba 0 0 0 1>}
  n114 = VectorParameter() {group=Wrap_SideMaskedEmissiveReactive_UVs, parameterName=SMER_Warp_Speed_Vehicle, defaultValue=<rgba 0 0 0 1>}
  n115 = VectorParameter() {group=Wrap_SideMaskedEmissiveReactive_UVs, parameterName=SMER_Warp_Speed_VehicleSide, defaultValue=<rgba 0 0 0 1>}
  n116 = MaterialFunctionCall(Color=@n6) {materialFunction=&SplitComponents}
  n117 = LinearInterpolate(A=@n113.RGB, B=LinearInterpolate(A=@n114.RGB, B=@n115.RGB, Alpha=@n116.G), Alpha=@n116.R)
  n118 = VectorParameter() {group=Wrap_SideMaskedEmissiveReactive_UVs, parameterName=SMER_Warp_Speed_Mech, defaultValue=<rgba 0 0 0 1>}
  n119 = ComponentMask(None=Multiply(A=Time(), B=LinearInterpolate(A=@n117, B=@n118.RGB, Alpha=@n116.B))) {g=true, r=true}
  n120 = ChannelMaskParameter(None=TextureSample(UVs=Add(A=@n112, B=@n119), Tex=@n4) {samplerType=SAMPLERTYPE_LinearGrayscale}.RGBA) {group=Wrap_SideMaskedEmissiveReactive_UVs, parameterName=SMER_Warp_Channel}
  n121 = ScalarParameter() {group=Wrap_SideMaskedEmissiveReactive_UVs, parameterName=SMER_Warp_Amount -----}
  StaticSwitchParameter(True=Add(A=@n13, B=Multiply(A=@n120.R, B=@n121)), False=@n13) {group=Wrap_SideMaskedEmissiveReactive_UVs, parameterName=Use Warp, sortPriority=-100}
  StaticSwitchParameter(True=@n1.Normal, False=@n102) {defaultValue=true, group=Wrap_SideMaskedEmissiveReactive, parameterName=SMER_Normal_Simple, sortPriority=-100}
 ]}
