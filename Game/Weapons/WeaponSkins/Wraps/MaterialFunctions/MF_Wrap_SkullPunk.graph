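# SkullPunk weapon wrap material function, layering primary and secondary looks over an incoming material by input masks, behind a static switch.
# Primary uses side-projected textures, curve-atlas gradients and camera-facing inner glow; secondary uses top/side projection with silk-like fresnel.
# Spectrum bands react to music, sine waves and Blueprint-set time.
# State inputs select projection transforms for weapons, vehicles, mechs, characters, backpacks and pickaxes; a cloth-fuzz branch has a mobile fallback.
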MaterialFunction /Game/Weapons/WeaponSkins/Wraps/MaterialFunctions/MF_Wrap_SkullPunk
graph {nodes [
  ActorPositionWS()
  n1 = NamedRerouteUsage()
  n2 = ScalarParameter() {defaultValue=0.01, group=Wrap_SkullPunk_Secondary, parameterName=SKP_Secondary_Normal_Scale}
  n3 = Time()
  n4 = VectorParameter() {group=Wrap_SkullPunk_Secondary, parameterName=SKP_Secondary_Normal_Speed, defaultValue=<rgba 0.123 0.03 0.083 1>}
  Add(
   A=Multiply(A=NamedRerouteUsage(), B=ScalarParameter() {defaultValue=2, group=Wrap_SkullPunk_Secondary, parameterName=SKP_Secondary_Normal_Warp})
   B=Add(A=Multiply(A=@n1, B=@n2), B=Multiply(A=@n3, B=@n4.RGB))
  )
  n5 = VectorParameter() {group=Wrap_SkullPunk_Primary, parameterName="SKP_Primary_TOD_Emissive_mulplyer A=M G=D B=E A=N", defaultValue=<rgba 0.5 0.5 0.5 0.25>}
  n6 = MaterialFunctionCall(Value_Morning=@n5.R, Value_Day=@n5.R, Value_Evening=@n5.R, Value_Night=@n5.R) {materialFunction=&Scalar_Time_of_Day}
  n7 = Add(A=NamedRerouteUsage(), B=ScalarParameter() {defaultValue=0.15, group=Wrap_SkullPunk_Primary, parameterName=SKP_Gradent_Gmap_influnce}) {constB=0.25}
  n8 = ChannelMaskParameter(None=NamedRerouteUsage()) {group=Wrap_SkullPunk_Primary, maskChannel=Blue, parameterName=SKP_Gradent_Z_Channel, defaultValue=<rgba 0 0 1 0>}
  n9 = Divide(A=@n8.R, B=ScalarParameter() {defaultValue=1, group=Wrap_SkullPunk_Primary, parameterName=SKP_Gradent_Z_fallof})
  n10 = TextureObjectParameter() {
   group=Wrap_SkullPunk_Primary
   parameterName=SKP_Primary__Normal
   samplerType=SAMPLERTYPE_Normal
   texture=&FlatNormalBlue
  }
  n11 = MaterialFunctionCall(Float4=NamedRerouteUsage()) {materialFunction=&BreakOutFloat4Components}
  n12 = MaterialFunctionCall!(
   For_Safe_Zones
   Normal_Map=@n10
   Caculated_Bounds=NamedRerouteUsage()
   Pre_Skinned_Normals=NamedRerouteUsage()
   Is Vehicle=@n11.R
   Is VehicleSide=@n11.R
   Is Mech=@n11.R
   Vehicle_Rotation
   UV_Rotation Bypass=NamedRerouteUsage()
   Mech_Rotation
   Uvs_Offset=NamedRerouteUsage()
   Offset-after rotation
   Uvs_TileRate=NamedRerouteUsage()
  ) {materialFunction=&/MF_Projection_Side_Normal}
  n13 = TextureSample(
   UVs=@n12.UVs
   Tex=TextureObjectParameter() {group=Wrap_SkullPunk_Primary, parameterName=SKP_Primary__Msk, texture=&T_BlacksRGB}
  )
  n14 = OneMinus(None=@n13.G)
  n15 = ScalarParameter() {group=Wrap_SkullPunk_Primary, parameterName=SKP_Gradent_DropShadow_Effect}
  n16 = Multiply(A=Multiply(A=Saturate(None=@n7), B=Saturate(None=@n9)), B=Saturate(None=Add(A=@n14, B=@n15) {constB=0}))
  n17 = CurveAtlasRowParameter(CurveTime=@n16) {
   atlas=/Game/Weapons/WeaponSkins/Wraps/Textures/CruveAtlas_Wraps_01
   curve=/Game/Weapons/WeaponSkins/Wraps/Textures/SkullPunk/CC_SKpunk_InnerGlow
   defaultValue=64
   group=Wrap_SkullPunk_Primary
   parameterName=SKP_Primary_Gmap_Color
  }
  n18 = MaterialFunctionCall!(
   Index=ScalarParameter() {defaultValue=64, group=Wrap_SkullPunk_Primary, parameterName="SKP_Primary_Gmap_Color  Index"}
   CCinput=@n17
   CCinput_Alpha=@n17
   UV=@n16
  ) {materialFunction=&/MF_Global_CC_Tex}
  n19 = Multiply(A=@n18.3V, B=@n18.3V)
  n20 = Multiply(A=@n13.A, B=NamedRerouteUsage())
  Add(A=LinearInterpolate(A=Multiply(A=@n6.Result, B=@n19), Alpha=@n13.R) {constB=0}, B=@n20)
  n21 = NamedRerouteUsage()
  n22 = Fresnel(ExponentIn=ScalarParameter() {defaultValue=1, group=Wrap_SkullPunk_Primary, parameterName=SKP_Fresnel_Expon}) {baseReflectFraction=0, exponent=1}
  n23 = CurveAtlasRowParameter(CurveTime=@n22) {
   atlas=&CruveAtlas_Wraps_01
   curve=/Game/Weapons/WeaponSkins/Wraps/Textures/SkullPunk/CC_SKpunk_Fres
   defaultValue=63
   group=Wrap_SkullPunk_Primary
   parameterName=SKP_Fresnel_Color
  }
  n24 = MaterialFunctionCall(
   Index=ScalarParameter() {defaultValue=63, group=Wrap_SkullPunk_Primary, parameterName="SKP_Fresnel_Color  Index"}
   CCinput=@n23
   CCinput_Alpha=@n23
   UV=@n22
  ) {materialFunction=&MF_Global_CC_Tex}
  n25 = Multiply(A=@n24.3V, B=@n24.3V)
  n26 = OneMinus(None=@n13.R)
  Add(A=Multiply(A=Add(A=@n21, B=@n25), B=@n26), B=@n20)
  n27 = NamedRerouteUsage()
  n28 = NamedRerouteUsage()
  n29 = ComponentMask(None=@n28) {g=true}
  n30 = Abs(None=Multiply(A=Add(A=@n29) {constB=-0.5}) {constB=2})
  n31 = Round(None=Add(A=@n30) {constB=0.2})
  Add(A=@n27, B=@n31)
  Add(A=@n28, B=Constant2Vector() {g=0.24, r=-0.495})
  CollectionParameter() {collection=&FortniteMaterialParameters, parameterName=EventAudio_Spectrum}
  n32 = VectorParameter() {group=Wrap_SkullPunk_UV_Primary, parameterName=WNL_Primary_Offset_Vehicle, defaultValue=<rgba 0 0 0 1>}
  n33 = VectorParameter() {group=Wrap_SkullPunk_UV_Primary, parameterName=WNL_Primary_Offset_VehicleSide, defaultValue=<rgba 0 0 0 1>}
  n34 = MaterialFunctionCall(Float4=NamedRerouteUsage()) {materialFunction=&BreakOutFloat4Components}
  n35 = VectorParameter() {group=Wrap_SkullPunk_UV_Primary, parameterName=WNL_Primary_Offset_SetA_Mech, defaultValue=<rgba 0 0 0 1>}
  n36 = VectorParameter() {group=Wrap_SkullPunk_UV_Primary, parameterName=WNL_Primary_Offset_Character, defaultValue=<rgba 0 0 0 1>}
  n37 = LinearInterpolate(
   A=LinearInterpolate(
    A=VectorParameter() {group=Wrap_SkullPunk_UV_Primary, parameterName=WNL_Primary_Offset_Weapon, defaultValue=<rgba 0 0 0 1>}.RGB
    B=LinearInterpolate(A=@n32.RGB, B=@n33.RGB, Alpha=@n34.G)
    Alpha=@n34.R
   )
   B=LinearInterpolate(A=@n35.RGB, B=@n36.RGB, Alpha=@n34.A)
   Alpha=@n34.B
  )
  ComponentMask(None=@n37) {g=true, r=true}
  Constant2Vector() {g=0.5, r=0.5}
  Constant3Vector()
  Constant3Vector()
  Constant4Vector()
  n38 = TextureSample() {samplerType=SAMPLERTYPE_Masks, texture=/Game/Weapons/FORT_Rifles/Textures/T_Scar_CM}
  n39 = FunctionInput(Preview=@n38.R) {inputName=Mask Primary, inputType=FunctionInput_Scalar, sortPriority=24, previewValue=<1 0 0 1>}
  n40 = FunctionInput(Preview=@n38.B) {inputName=Mask Secondary, inputType=FunctionInput_Scalar, sortPriority=24, previewValue=<1 0 0 1>}
  n41 = BlendMaterialAttributes(
   A=BlendMaterialAttributes(A=NamedRerouteUsage(), B=NamedRerouteUsage(), Alpha=@n39) {vertexAttributeBlendType=UseA}
   B=NamedRerouteUsage()
   Alpha=@n40
  ) {vertexAttributeBlendType=UseA}
  n42 = TextureSample() {samplerType=SAMPLERTYPE_Normal, texture=/Game/Weapons/FORT_Rifles/T_SCAR_N}
  n43 = Constant3Vector() {constant=<rgba 1 0 0 0>}
  n44 = GetMaterialAttributes(None=FunctionInput(Preview=SetMaterialAttributes!(MaterialAttributes, Normal=@n42.RGB, BaseColor=@n43) {attributeSetTypes [#1, #2]}) {inputName=Material, inputType=FunctionInput_MaterialAttributes}) {attributeGetTypes [
    #2
    #3
    #1
    #4
   ]}
  n45 = NamedRerouteDeclaration(None=@n44.MaterialAttributes) {name=Root_Material, nodeColor=<rgba 0.7812 1 0 1>}
  FunctionOutput(None=StaticSwitchParameter(True=@n41, False=@n45) {group=Customization_Wrap, parameterName=Wrap_SkullPunk})
  n46 = TextureObjectParameter() {
   group=Wrap_SkullPunk_Secondary
   parameterName=SKP_Secondary__Normal
   samplerType=SAMPLERTYPE_Normal
   texture=&FlatNormalBlue
  }
  n47 = NamedRerouteUsage()
  n48 = Add(A=Multiply(A=@n1, B=NamedRerouteUsage()), B=NamedRerouteUsage())
  n49 = NamedRerouteUsage()
  n50 = ScalarParameter() {defaultValue=1, group=Wrap_SkullPunk_Secondary, parameterName=SKP_Secondary_Blend_Contrast}
  n51 = ScalarParameter() {defaultValue=1, group=Wrap_SkullPunk_Secondary, parameterName=SKP_Secondary_Fill_alternate_Axis_Opacity}
  n52 = MaterialFunctionCall(Float4=NamedRerouteUsage()) {materialFunction=&BreakOutFloat4Components}
  n53 = MaterialFunctionCall!(
   Texture_Normals=@n46
   Rotation_Axis=@n47
   Caculated_Bounds=@n48
   Pre_Skinned_Normals=@n49
   Blend_Contrast=@n50
   Fill_alternate_Axis_Opacity=@n51
   Is_Vehicle=@n52.R
   Is_Mech=@n52.R
  ) {materialFunction=&/MF_TopsSideProjection_Normals}
  n54 = VectorParameter() {group=Wrap_SkullPunk_Secondary, parameterName=SKP_Secondary_Fill_alternate_Axis}
  n55 = MaterialFunctionCall(Float4=MaterialFunctionCall!(
    IsCharacter=@n52.A
    IsStatsB=NamedRerouteUsage()
    Texture=TextureObjectParameter() {group=Wrap_SkullPunk_Secondary, parameterName=SKP_Secondary__Msk, texture=&T_BlacksRGB}
    Rotation_Axis=@n47
    Caculated_Bounds=@n48
    Pre_Skinned_Normals=@n49
    Blend_Contrast=@n50
    Fill_alternate_Axis_Opacity=@n51
    "Fill_alternate_Axis_Color_A=Roughness"=AppendVector(A=@n54.RGB, B=@n54.RGB)
    Is_Vehicle=@n52.A
    Is_Mech=@n52.A
    Enable_Fill_Z_Weapon
    Fill_Z_Weapon_opacity
    Fill_Z_Weapon_Color
   ) {materialFunction=&/MF_TopsSideProjection_Base}.Combined) {materialFunction=&BreakOutFloat4Components}
  LinearInterpolate(A=Constant3Vector(), B=@n53.Add_Surface_Normal, Alpha=@n55.A)
  n56 = ComponentMask(None=NamedRerouteUsage()) {g=true, r=true}
  n57 = ScalarParameter() {defaultValue=0.25, group=Wrap_SkullPunk_Primary, parameterName=SKP_Primary_InnerGlow_Warp}
  n58 = PreSkinnedLocalBounds()
  n59 = Add(
   A=@n58.Min
   B=FunctionInput(Preview=Constant3Vector()) {bUsePreviewValueAsDefault=true, inputName=Object Bounds Min Adjustment, sortPriority=50}
  ) {constB=0}
  n60 = FunctionInput(Preview=Constant3Vector()) {bUsePreviewValueAsDefault=true, inputName=Object Bounds Max Adjustment, sortPriority=50}
  n61 = Multiply(A=Subtract(A=Add(A=@n58.Max, B=@n60) {constB=0}, B=@n59)) {constB=0.5}
  n62 = MaterialFunctionCall(Float4=NamedRerouteUsage()) {materialFunction=&BreakOutFloat4Components}
  n63 = LinearInterpolate(A=TransformPosition(None=Add(A=@n59, B=@n61)) {transformType=TRANSFORMPOSSOURCE_World}, B=ObjectPositionWS(), Alpha=@n62.B)
  n64 = Normalize(VectorInput=Subtract(A=@n63, B=CameraPositionWS()))
  n65 = Constant3Vector() {constant=<rgba 0 0 1 1>}
  n66 = MaterialFunctionCall(Float3=@n58.Extents) {materialFunction=&BreakOutFloat3Components}
  n67 = Max(A=@n66.R, B=@n66.R)
  n68 = MaterialFunctionCall!(
   Plane Normal Axis=@n64
   Plane X Axis=Normalize(VectorInput=CrossProduct(A=@n64, B=@n65))
   Plane Center=@n63
   UV Scale=Multiply(
    A=MaterialFunctionCall() {materialFunction=/Engine/Functions/Engine_MaterialFunctions02/WorldPositionOffset/ObjectScale}.Scale X
    B=Max(A=@n67, B=@n66.B)
   )
  ) {materialFunction=/Game/Athena/Environments/Nexus/Materials/VirtualPlaneCoordinatesAbs}
  n69 = FunctionInput(Preview=Constant4Vector() {constant=<rgba 0 0 0 1>}.RGBA) {bUsePreviewValueAsDefault=true, inputName=Object-SafeZone-Virtual-Pivot, inputType=FunctionInput_Vector4, sortPriority=105}
  n70 = FunctionInput(Preview=Constant()) {bUsePreviewValueAsDefault=true, inputName=Object-SafeZone-VP_Charcter_Preview_Switch, inputType=FunctionInput_Scalar, sortPriority=105}
  n71 = MaterialFunctionCall!(
   Camera _pos
   Axis
   Scale
   Scale_Multplyer
   Object-SafeZone-Virtual-Pivot=@n69
   Object-SafeZone-Virtual-CH-Switch=@n70
   SinglePlane Normal Axis
   SinglePlane X Axis
   PlaneCenter_distortion
  ) {materialFunction=&/MF_SafeZone_VirtualPlane}
  n72 = SphereMask(
   A=Add(A=Multiply(A=@n56, B=@n57), B=LinearInterpolate(A=@n68.UVs, B=@n71.Uvs, Alpha=@n62.A))
   B=Constant() {r=0.5}
   Radius=ScalarParameter() {defaultValue=0.5, group=Wrap_SkullPunk_Primary, parameterName=SKP_Primary_InnerGlow_Radius}
  ) {hardnessPercent=0}
  LinearInterpolate(
   A=VectorParameter() {group=Wrap_SkullPunk_Primary, parameterName=SKP_Primary_InnerGlow_Color_A, defaultValue=<rgba 0 0 0 1>}.RGB
   B=VectorParameter() {group=Wrap_SkullPunk_Primary, parameterName=SKP_Primary_InnerGlow_Color_B, defaultValue=<rgba 0 0 0 1>}.RGB
   Alpha=@n72
  )
  n73 = TransformPosition(None=Constant3Vector()) {transformType=TRANSFORMPOSSOURCE_World}
  n74 = Normalize(VectorInput=Subtract(A=@n73, B=CameraPositionWS()))
  n75 = CrossProduct(A=@n74, B=Constant3Vector() {constant=<rgba 0 0 1 1>})
  n76 = ComponentMask(None=@n69) {b=true, g=true, r=true}
  n77 = Add(A=@n73, B=@n76)
  n78 = ComponentMask(None=@n69) {a=true}
  n79 = MaterialFunctionCall(Float3=PreSkinnedLocalBounds().Extents) {materialFunction=&BreakOutFloat3Components}
  n80 = Multiply(
   A=MaterialFunctionCall() {materialFunction=&ObjectScale}.Scale X
   B=Max(A=Max(A=@n79.R, B=@n79.R), B=@n79.B)
  )
  Multiply(A=MaterialFunctionCall(
    Plane Normal Axis=@n74
    Plane X Axis=Normalize(VectorInput=@n75)
    Plane Center=LinearInterpolate(A=@n77, B=@n76, Alpha=@n70)
    UV Scale=Multiply(A=@n78, B=@n80)
   ) {materialFunction=&VirtualPlaneCoordinatesAbs}.UVs)
  Multiply(A=@n55.A)
  n81 = Time()
  n82 = ScalarParameter() {defaultValue=-0.1, group=Wrap_SkullPunk_Spectrum, parameterName=SKP__Specturm_Color_Speed}
  n83 = Multiply(A=@n28, B=Constant() {r=3})
  n84 = Frac(None=Add(A=Multiply(A=@n81, B=@n82), B=ComponentMask(None=@n83) {r=true}))
  n85 = CurveAtlasRowParameter(CurveTime=@n84) {
   atlas=&CruveAtlas_Wraps_01
   curve=/Game/Weapons/WeaponSkins/Wraps/Textures/SkullPunk/CC_SKpunk_Spec
   defaultValue=65
   group=Wrap_SkullPunk_Spectrum
   parameterName=SKP__Specturm _Color
  }
  n86 = MaterialFunctionCall(Index=ScalarParameter() {defaultValue=65, group=Wrap_SkullPunk_Spectrum, parameterName="SKP__Specturm _Color  Index"}, CCinput=@n85, UV=@n84) {materialFunction=&MF_Global_CC_Tex}
  n87 = ScalarParameter() {defaultValue=0.32, group=Wrap_SkullPunk_Spectrum, parameterName=SKP_Specturm_A_Mult}
  n88 = Subtract(A=Time(), B=ScalarParameter() {defaultValue=-9999999, group=Wrap_TimefromBP, parameterName=TimeFrom-BP----------------, sortPriority=600})
  n89 = Saturate(None=Divide(A=@n88, B=ScalarParameter() {defaultValue=0.7, group=Wrap_TimefromBP, parameterName=SKP_Time_Duration_A, sortPriority=601}))
  n90 = MaterialFunctionCall(Float4=Multiply(
    A=CollectionParameter() {collection=/Game/Athena/Events/AudioAnalysis/MPC_Audio_Reactive_Parameters, parameterName=Amplitude_Music}
    B=CollectionParameter() {collection=&MPC_Audio_Reactive_Parameters, parameterName=Spectrum_Music}
   )) {materialFunction=&BreakOutFloat4Components}
  n91 = VectorParameter() {group=Wrap_SkullPunk_Spectrum, parameterName=SKP_Specturm_Speeds A-C, defaultValue=<rgba 0.5 0.5 0.5 0.5>}
  n92 = Sine(None=Multiply(A=Time(), B=@n91.R))
  n93 = Max(A=Max(A=OneMinus(None=@n89), B=@n90.R), B=Multiply(A=Add(A=@n92), B=@n91.A) {constB=0.5})
  n94 = Add(A=ComponentMask(None=@n28) {r=true}, B=Constant() {r=-0.98})
  n95 = ScalarParameter() {defaultValue=0.001, group=Wrap_SkullPunk_Spectrum, parameterName=SKP_Specturm_Falloff}
  n96 = Divide(A=Add(A=Multiply(A=@n87, B=@n93), B=@n94), B=@n95)
  n97 = ScalarParameter() {defaultValue=0.4, group=Wrap_TimefromBP, parameterName=SKP_Time_Duration_B, sortPriority=601}
  n98 = OneMinus(None=Saturate(None=Divide(A=@n88, B=@n97)))
  n99 = Time()
  n100 = Add(A=Sine(None=Multiply(A=@n99, B=@n91.G)))
  n101 = Multiply(
   A=ScalarParameter() {defaultValue=0.2, group=Wrap_SkullPunk_Spectrum, parameterName=SKP_Specturm_B_Mult}
   B=Max(A=Max(A=@n90.B, B=@n98), B=Multiply(A=@n100, B=@n91.A) {constB=0.5})
  )
  n102 = Saturate(None=Divide(A=Add(A=@n101, B=@n94), B=@n95))
  n103 = ScalarParameter() {defaultValue=0.5, group=Wrap_TimefromBP, parameterName=SKP_Time_Duration_C, sortPriority=601}
  n104 = OneMinus(None=Saturate(None=Divide(A=@n88, B=@n103)))
  n105 = Time()
  n106 = Add(A=Sine(None=Multiply(A=@n105, B=@n91.B)))
  n107 = Multiply(
   A=ScalarParameter() {defaultValue=0.21, group=Wrap_SkullPunk_Spectrum, parameterName=SKP_Specturm_C_Mult}
   B=Max(A=Max(A=@n90.G, B=@n104), B=Multiply(A=@n106, B=@n91.A) {constB=0.5})
  )
  n108 = Saturate(None=Divide(A=Add(A=@n107, B=@n94), B=@n95))
  n109 = Round(None=Add(A=@n30) {constB=0.15})
  n110 = LinearInterpolate(
   A=Multiply(A=@n86.3V) {constB=0.3}
   B=Multiply(A=@n86.3V, B=ScalarParameter() {defaultValue=2, group=Wrap_SkullPunk_Spectrum, parameterName=SKP_Specturm__Brightness})
   Alpha=LinearInterpolate(A=Saturate(None=@n96), B=LinearInterpolate(A=@n102, B=@n108, Alpha=@n109), Alpha=@n31)
  )
  Multiply(A=@n110, B=@n27)
  n111 = VectorParameter() {group=Wrap_SkullPunk_UV_2nd, parameterName=SKP_Rotation_Weapon, defaultValue=<rgba 0 0 -0.25 1>}
  n112 = VectorParameter() {desc=must activate TS_Vehicle_Side_01 UVs, group=Wrap_SkullPunk_UV_2nd, parameterName=SKP_Rotation_Vehicle_Side, defaultValue=<rgba 0 0 0 1>}
  n113 = MaterialFunctionCall(Float4=NamedRerouteUsage()) {materialFunction=&BreakOutFloat4Components}
  n114 = LinearInterpolate(
   A=VectorParameter() {group=Wrap_SkullPunk_UV_2nd, parameterName=SKP_Rotation_Vehicle, defaultValue=<rgba 0 0 0 1>}.RGB
   B=@n112.RGB
   Alpha=@n113.G
  )
  n115 = VectorParameter() {group=Wrap_SkullPunk_UV_2nd, parameterName=SKP_Rotation_Mech, defaultValue=<rgba 0.25 0.25 0 1>}
  n116 = VectorParameter() {group=Wrap_SkullPunk_UV_2nd, parameterName=SKP_Rotation_Characte, defaultValue=<rgba 0.25 0.25 0 1>}
  NamedRerouteDeclaration(None=LinearInterpolate(A=LinearInterpolate(A=@n111.RGB, B=@n114, Alpha=@n113.R), B=LinearInterpolate(A=@n115.RGB, B=@n116.RGB, Alpha=@n113.A), Alpha=@n113.B)) {name=TopSide_Rotation, nodeColor=<rgba 0.9922 1 0 1>}
  n117 = VectorParameter() {group=Wrap_SkullPunk_UV_2nd, parameterName=SKP_Offset_Vehicle_01, defaultValue=<rgba 0 0 0 1>}
  n118 = VectorParameter() {group=Wrap_SkullPunk_UV_2nd, parameterName=SKP_Offset_VehicleSide_01, defaultValue=<rgba 0 0 0 1>}
  n119 = MaterialFunctionCall(Float4=NamedRerouteUsage()) {materialFunction=&BreakOutFloat4Components}
  n120 = LinearInterpolate(
   A=LinearInterpolate(
    A=VectorParameter() {group=Wrap_SkullPunk_UV_2nd, parameterName=SKP_Offset_Weapon_01, defaultValue=<rgba 0 0 0 1>}.RGB
    B=LinearInterpolate(A=@n117.RGB, B=@n118.RGB, Alpha=@n119.G)
    Alpha=@n119.R
   )
   B=LinearInterpolate(
    A=VectorParameter() {group=Wrap_SkullPunk_UV_2nd, parameterName=SKP_Offset_Mech_01, defaultValue=<rgba 0 0 0 1>}.RGB
    B=VectorParameter() {group=Wrap_SkullPunk_UV_2nd, parameterName=SKP_Offset_Character_01, defaultValue=<rgba 0 0 0 1>}.RGB
    Alpha=@n119.A
   )
   Alpha=@n119.B
  )
  n121 = VectorParameter() {group=Wrap_SkullPunk_UV_2nd, parameterName=SKP_Offset_Backpack_01-----------, defaultValue=<rgba 0 0 0 1>}
  n122 = MaterialFunctionCall(Float2=NamedRerouteUsage()) {materialFunction=&BreakOutFloat2Components}
  NamedRerouteDeclaration(None=LinearInterpolate(
    A=LinearInterpolate(A=@n120, B=@n121.RGB, Alpha=@n122.R)
    B=VectorParameter() {group=Wrap_SkullPunk_UV_2nd, parameterName=SKP_Offset_Pickaxe_01, defaultValue=<rgba 0 0 0 1>}.RGB
    Alpha=@n122.G
   )) {name=TopSide_Offsets, nodeColor=<rgba 0 0.625 1 1>}
  n123 = VectorParameter() {group=Wrap_SkullPunk_Secondary, parameterName=SKP_Secondary_Base_Color_A, defaultValue=<rgba 0.4427 0.2733 0.083 0.4>}
  n124 = VectorParameter() {group=Wrap_SkullPunk_Secondary, parameterName=SKP_Secondary_Base_Color_B, defaultValue=<rgba 0.2917 0.1668 0.0327 0.1>}
  n125 = SetMaterialAttributes!(
   MaterialAttributes=NamedRerouteUsage()
   BaseColor=LinearInterpolate(A=@n123.RGB, B=@n124.RGB, Alpha=@n55.A)
   Roughness=LinearInterpolate(A=@n123.A, B=@n124.A, Alpha=@n55.A)
   Specular=ScalarParameter() {defaultValue=0.5, group=Wrap_SkullPunk_Secondary, parameterName=SKP_Secondary_Specular}
   Metallic=ScalarParameter() {defaultValue=1, group=Wrap_SkullPunk_Secondary, parameterName=SKP_Secondary_Metalic}
   Normal=Add(A=NamedRerouteUsage(), B=@n53.Add_Surface_Normal)
  ) {attributeSetTypes [
    #2
    #3
    #5
    #4
    #1
   ]}
  n126 = GetMaterialAttributes(None=@n125) {attributeGetTypes [
    #2
    #4
    #5
    #3
   ]}
  n127 = ScalarParameter() {defaultValue=1, group=Wrap_SkullPunk_Secondary, parameterName=SKP_Silk_BaseColorBrightness}
  n128 = VectorParameter() {group=Wrap_SkullPunk_Secondary, parameterName=SKP_Silk_EdgeTint, defaultValue=<rgba 0.1094 0.1094 0.1094 1>}
  n129 = Fresnel(ExponentIn=ScalarParameter() {defaultValue=2, group=Wrap_SkullPunk_Secondary, parameterName=SKP_Silk_Fresnel}) {baseReflectFraction=0}
  n130 = ScalarParameter() {defaultValue=0.5, group=Wrap_SkullPunk_Secondary, parameterName=SKP_Silk_EdgeAniso}
  n131 = VectorParameter() {group=Wrap_SkullPunk_Secondary, parameterName=SKP_Silk_EdgeGlow, defaultValue=<rgba 0 0 0 1>}
  n132 = Constant() {r=0.5}
  n133 = Constant() {r=1}
  n134 = Constant() {r=0.5}
  n135 = Constant() {r=0.25}
  n136 = SetMaterialAttributes!(
   MaterialAttributes=@n126.MaterialAttributes
   BaseColor=LinearInterpolate(A=Multiply(A=@n126.BaseColor, B=@n127), B=@n128.RGB, Alpha=@n129)
   Metallic=LinearInterpolate(
    A=ScalarParameter() {defaultValue=1, group=Wrap_SkullPunk_Secondary, parameterName=SKP_Silk_FresnelMax}
    B=ScalarParameter() {group=Wrap_SkullPunk_Secondary, parameterName=SKP_Silk_FresnelMin}
    Alpha=@n129
   )
   Specular=@n126.MaterialAttributes
   Roughness=LinearInterpolate(A=@n126.Roughness, B=Multiply(A=@n126.Roughness, B=@n130), Alpha=@n129)
   ShadingModel=ShadingModel()
   SubsurfaceColor=@n126.MaterialAttributes
   CustomData0=Constant() {r=1}
   EmissiveColor=Multiply(
    A=Multiply(A=@n129, B=@n131.RGB)
    B=MaterialFunctionCall(Value_Morning=@n132, Value_Day=@n133, Value_Evening=@n134, Value_Night=@n135) {materialFunction=&Scalar_Time_of_Day}.Result
   )
  ) {attributeSetTypes [
    #2
    #4
    #5
    #3
    #6
    #7
    #8
    #9
   ]}
  NamedRerouteDeclaration(None=@n136) {name=2nd_Material, nodeColor=<rgba 0 1 0.5078 1>}
  n137 = ScalarParameter() {group=Wrap_SkullPunk_UV_Primary, parameterName=WNL_Primary_Rotation_Weapon}
  n138 = MaterialFunctionCall(Float4=NamedRerouteUsage()) {materialFunction=&BreakOutFloat4Components}
  n139 = LinearInterpolate(
   A=ScalarParameter() {group=Wrap_SkullPunk_UV_Primary, parameterName=WNL_Primary_Rotation_Vehicle}
   B=ScalarParameter() {group=Wrap_SkullPunk_UV_Primary, parameterName=WNL_Primary_Rotation_VehicleSide}
   Alpha=@n138.G
  )
  n140 = ScalarParameter() {defaultValue=0.25, group=Wrap_SkullPunk_UV_Primary, parameterName=WNL_Primary_Rotation_Mech}
  n141 = ScalarParameter() {group=Wrap_SkullPunk_UV_Primary, parameterName=WNL_Primary_Rotation_Character}
  NamedRerouteDeclaration(None=LinearInterpolate(A=LinearInterpolate(A=@n137, B=@n139, Alpha=@n138.R), B=LinearInterpolate(A=@n140, B=@n141, Alpha=@n138.A), Alpha=@n138.B)) {name=Primary_Rotation, nodeColor=<rgba 1 0.7266 0 1>}
  n142 = ScalarParameter() {defaultValue=1, group=Wrap_SkullPunk_UV_Primary, parameterName=WNL_Primary_Scale_Weapon}
  n143 = MaterialFunctionCall(Float4=NamedRerouteUsage()) {materialFunction=&BreakOutFloat4Components}
  n144 = LinearInterpolate(
   A=ScalarParameter() {defaultValue=1, group=Wrap_SkullPunk_UV_Primary, parameterName=WNL_Primary_Scale_Vehicle}
   B=ScalarParameter() {defaultValue=1, group=Wrap_SkullPunk_UV_Primary, parameterName=WNL_Primary_Scale_VehicleSide}
   Alpha=@n143.G
  )
  n145 = ScalarParameter() {defaultValue=1, group=Wrap_SkullPunk_UV_Primary, parameterName=WNL_Primary_Scale_Mech}
  n146 = ScalarParameter() {defaultValue=1, group=Wrap_SkullPunk_UV_Primary, parameterName=WNL_Primary_Scale_Character}
  NamedRerouteDeclaration(None=LinearInterpolate(A=LinearInterpolate(A=@n142, B=@n144, Alpha=@n143.R), B=LinearInterpolate(A=@n145, B=@n146, Alpha=@n143.A), Alpha=@n143.B)) {name=Primary_Scale, nodeColor=<rgba 1 0.7266 0 1>}
  n147 = CurveAtlasRowParameter(CurveTime=@n72) {
   atlas=&CruveAtlas_Wraps_01
   curve=&CC_SKpunk_InnerGlow
   defaultValue=64
   group=Wrap_SkullPunk_Primary
   parameterName=SKP_Primary_InnerGlow_Color
  }
  n148 = MaterialFunctionCall(
   Index=ScalarParameter() {defaultValue=64, group=Wrap_SkullPunk_Primary, parameterName="SKP_Primary_InnerGlow_Color  Index"}
   CCinput=@n147
   CCinput_Alpha=@n147
   UV=@n72
  ) {materialFunction=&MF_Global_CC_Tex}
  NamedRerouteDeclaration(None=Multiply(A=@n148.3V, B=@n148.3V)) {name=Primary_InnerGlow, nodeColor=<rgba 0 0.1328 1 1>}
  NamedRerouteDeclaration(None=@n12.UVs) {name=SideProject_UVS, nodeColor=<rgba 0 1 0.8359 1>}
  NamedRerouteDeclaration(None=@n110) {name=Spectrum, nodeColor=<rgba 0.6406 0 1 1>}
  n149 = TextureSample(
   UVs=@n12.UVs
   Tex=TextureObjectParameter() {group=Wrap_SkullPunk_Primary, parameterName=SKP_Primary__BC, texture=&T_BlacksRGB}
  )
  n150 = Multiply(A=@n149.A, B=Add(A=OneMinus(None=@n13.A)) {constB=0.25})
  n151 = NamedRerouteUsage()
  n152 = SetMaterialAttributes!(
   MaterialAttributes=NamedRerouteUsage()
   BaseColor=LinearInterpolate(A=@n19, B=@n149.RGB, Alpha=@n13.R)
   Roughness=LinearInterpolate(
    A=ScalarParameter() {group=Wrap_SkullPunk_Primary, parameterName=SKP_Primary_Roghness_Low}
    B=ScalarParameter() {defaultValue=1, group=Wrap_SkullPunk_Primary, parameterName=SKP_Primary_Roghness_High}
    Alpha=Saturate(None=@n150)
   )
   Specular=LinearInterpolate(
    A=ScalarParameter() {defaultValue=0.5, group=Wrap_SkullPunk_Primary, parameterName=SKP_Primary_Specular_Emissive}
    B=ScalarParameter() {defaultValue=0.25, group=Wrap_SkullPunk_Primary, parameterName=SKP_Primary_Specular}
    Alpha=@n13.R
   )
   EmissiveColor=@n20
   Metallic=ScalarParameter() {group=Wrap_SkullPunk_Primary, parameterName=SKP_Primary_Metallic}
   Normal=NamedRerouteDeclaration(None=Add(A=@n151, B=@n12.Raw_Normal)) {name=Primary_Normals, nodeColor=<rgba 0.6406 1 0 1>}
  ) {attributeSetTypes [
    #2
    #3
    #5
    #9
    #4
    #1
   ]}
  NamedRerouteDeclaration(None=@n152) {name=1st_Material, nodeColor=<rgba 0 0.3906 1 1>}
  NamedRerouteDeclaration(None=FunctionInput(Preview=@n38.A) {inputName=Mask Gmap, inputType=FunctionInput_Scalar, sortPriority=24, previewValue=<1 0 0 1>}) {name=Mask_Gmap, nodeColor=<rgba 1 0.7222 0.1142 1>}
  NamedRerouteDeclaration(None=@n37) {name=Primary_Offset, nodeColor=<rgba 1 0.7266 0 1>}
  NamedRerouteDeclaration(None=@n13.A) {name=Teamp_mask, nodeColor=<rgba 0 1 0.5078 1>}
  NamedRerouteDeclaration(None=VectorParameter() {group=Wrap_SkullPunk_ClothFuzz, parameterName=SKP_Cloth Fuzz Tint, defaultValue=<rgba 1 1 1 1>}.RGB) {name=Cloth Color, nodeColor=<rgba 0 1 0.2969 1>}
  n153 = GetMaterialAttributes(None=NamedRerouteUsage()) {attributeGetTypes [#2, #1, #3]}
  NamedRerouteDeclaration(None=@n153.MaterialAttributes) {name=Root2_Material, nodeColor=<rgba 0.5703 1 0 1>}
  NamedRerouteDeclaration(None=@n26) {name=Alpha Source, nodeColor=<rgba 1 0.2109 0 1>}
  NamedRerouteDeclaration(None=@n41) {name=Output_material, nodeColor=<rgba 1 0 0.8906 1>}
  n154 = NamedRerouteUsage()
  n155 = NamedRerouteUsage()
  n156 = Multiply(A=@n153.BaseColor, B=ScalarParameter() {defaultValue=1, group=Wrap_SkullPunk_ClothFuzz, parameterName=SKP_Cloth Subsurface Intensity})
  n157 = ScalarParameter() {defaultValue=1, group=Wrap_SkullPunk_ClothFuzz, parameterName=SKP_Cloth Roughness}
  n158 = Multiply(A=@n153.BaseColor, B=ScalarParameter() {defaultValue=0.9, group=Wrap_SkullPunk_ClothFuzz, parameterName=SKP_Cloth Base Color Intensity})
  n159 = DotProduct(A=PixelNormalWS(), B=CameraVectorWS())
  n160 = Saturate(None=Power(Base=OneMinus(None=@n159), Exp=ScalarParameter() {defaultValue=4, group=Wrap_SkullPunk_ClothFuzz, parameterName=SKP_Fuzz Exponent}))
  n161 = TextureSampleParameter2D(UVs=Multiply(A=TextureCoordinate(), B=ScalarParameter() {defaultValue=4, group=Wrap_SkullPunk_ClothFuzz, parameterName=SKP_Fuzz Tiling})) {
   group=Wrap_SkullPunk_ClothFuzz
   parameterName=SKP_ClothFuzz Texture
   samplerSource=SSM_Wrap_WorldGroupSettings
   samplerType=SAMPLERTYPE_LinearColor
   texture=/Game/Global/Textures/Default/Masks_and_Gradients/T_Fuzz_MASK
  }
  n162 = Saturate(None=Multiply(A=Multiply(A=@n160, B=@n161.R), B=ScalarParameter() {defaultValue=1, group=Wrap_SkullPunk_ClothFuzz, parameterName=SKP_Fuzz Fresnel Blend}))
  n163 = Multiply(A=NamedRerouteUsage(), B=Add(A=Multiply(A=@n162) {constB=10}))
  n164 = SetMaterialAttributes!(
   MaterialAttributes=NamedRerouteUsage()
   ShadingModel=ShadingModel() {shadingModel=MSM_Cloth}
   SubsurfaceColor=Saturate(None=Multiply(A=@n155, B=@n156))
   Roughness=Saturate(None=Multiply(A=@n153.Roughness, B=@n157))
   CustomData0=ScalarParameter() {defaultValue=1, group=Wrap_SkullPunk_ClothFuzz, parameterName=SKP_Cloth Shader Intensity}
   Opacity=Constant() {r=1}
   BaseColor=Saturate(None=Multiply(A=@n158, B=@n163))
   Specular=Saturate(None=ScalarParameter() {defaultValue=0.75, group=Wrap_SkullPunk_ClothFuzz, parameterName=SKP_Cloth Scattering})
   Normal=@n153.Normal
  ) {attributeSetTypes [
    #6
    #7
    #3
    #8
    #10
    #2
    #5
    #1
   ]}
  n165 = NamedRerouteUsage()
  n166 = NamedRerouteUsage()
  n167 = NamedRerouteUsage()
  n168 = SetMaterialAttributes!(MaterialAttributes=NamedRerouteUsage(), BaseColor=Saturate(None=Multiply(A=@n153.BaseColor, B=@n167))) {attributeSetTypes [#2]}
  n169 = NamedRerouteUsage()
  NamedRerouteDeclaration(None=ShadingPathSwitch(
    Default=BlendMaterialAttributes(A=@n154, B=@n164, Alpha=@n165) {vertexAttributeBlendType=UseA}
    Mobile=BlendMaterialAttributes(A=@n166, B=@n168, Alpha=@n169) {vertexAttributeBlendType=UseA}
   )) {name=Cloth_output, nodeColor=<rgba 1 0 0.3516 1>}
  NamedRerouteDeclaration(None=@n55.R) {name=SecondaryWarp, nodeColor=<rgba 0.1016 1 0 1>}
  NamedRerouteDeclaration(None=FunctionInput(Preview=Constant4Vector().RGBA) {bUsePreviewValueAsDefault=true, inputName=IsState, inputType=FunctionInput_Vector4, sortPriority=4}) {name=IsState, nodeColor=<rgba 0.1497 1 0.1633 1>}
  NamedRerouteDeclaration(None=FunctionInput(Preview=Constant2Vector()) {bUsePreviewValueAsDefault=true, inputName=IsState_B, inputType=FunctionInput_Vector2, sortPriority=4}) {name=IsState_B, nodeColor=<rgba 0.1497 1 0.1633 1>}
  n170 = PreSkinnedLocalBounds()
  n171 = Subtract(A=VertexInterpolator(VS=PreSkinnedPosition()).PS, B=@n170.Min)
  n172 = MaterialFunctionCall(Color=Subtract(A=@n170.Max, B=@n170.Max)) {materialFunction=&SplitComponents}
  n173 = Max(A=Max(A=@n172.R, B=@n172.R), B=@n172.B)
  NamedRerouteDeclaration(None=FunctionInput(Preview=Divide(A=@n171, B=@n173)) {bUsePreviewValueAsDefault=true, inputName=Calcated_Bounds, sortPriority=11}) {name=Caclcated_Bounds, nodeColor=<rgba 0 0.9297 1 1>}
  n174 = VectorParameter() {group=Wrap_SkullPunk_UV_2nd, parameterName=SKP_Scale2nd_Weapon_01, defaultValue=<rgba 1 1 1 4>}
  n175 = VectorParameter() {group=Wrap_SkullPunk_UV_2nd, parameterName=SKP_Scale2nd_Vehicle_01, defaultValue=<rgba 1 1 1 1>}
  n176 = Multiply(A=@n175.RGB, B=@n175.RGB)
  n177 = VectorParameter() {group=Wrap_SkullPunk_UV_2nd, parameterName=SKP_Scale2nd_VehicleSide_01, defaultValue=<rgba 1 1 1 1>}
  n178 = Multiply(A=@n177.RGB, B=@n177.RGB)
  n179 = MaterialFunctionCall(Float4=NamedRerouteUsage()) {materialFunction=&BreakOutFloat4Components}
  n180 = VectorParameter() {group=Wrap_SkullPunk_UV_2nd, parameterName=SKP_Scale2nd_Mech_01, defaultValue=<rgba 1 1 1 1>}
  n181 = VectorParameter() {group=Wrap_SkullPunk_UV_2nd, parameterName=SKP_Scale2nd_Character_01, defaultValue=<rgba 1 1 1 1>}
  n182 = LinearInterpolate(
   A=LinearInterpolate(A=Multiply(A=@n174.RGB, B=@n174.RGB), B=LinearInterpolate(A=@n176, B=@n178, Alpha=@n179.G), Alpha=@n179.R)
   B=LinearInterpolate(A=Multiply(A=@n180.RGB, B=@n180.RGB), B=Multiply(A=@n181.RGB, B=@n181.RGB), Alpha=@n179.A)
   Alpha=@n179.B
  )
  n183 = VectorParameter() {group=Wrap_SkullPunk_UV_2nd, parameterName=SKP_Scale2nd_Backpack_01-----------, defaultValue=<rgba 1 1 1 1>}
  n184 = Multiply(A=@n183.RGB, B=@n183.RGB)
  n185 = MaterialFunctionCall(Float2=NamedRerouteUsage()) {materialFunction=&BreakOutFloat2Components}
  n186 = VectorParameter() {group=Wrap_SkullPunk_UV_2nd, parameterName=SKP_Scale2nd_Pickaxe_01, defaultValue=<rgba 1 1 1 1>}
  NamedRerouteDeclaration(None=LinearInterpolate(A=LinearInterpolate(A=@n182, B=@n184, Alpha=@n185.R), B=Multiply(A=@n186.RGB, B=@n186.RGB), Alpha=@n185.G)) {name=TopSide_Scale_2nd, nodeColor=<rgba 0.4062 0 1 1>}
  n187 = PreSkinnedNormal()
  NamedRerouteDeclaration(None=FunctionInput(Preview=VertexInterpolator(VS=@n187).PS) {bUsePreviewValueAsDefault=true, inputName=Pre-Skinned Local Normal, sortPriority=10}) {name=PreSkin_Normal, nodeColor=<rgba 0.8281 1 0 1>}
  NamedRerouteDeclaration(None=@n44.Normal) {name=Root_Normal, nodeColor=<rgba 0.4062 0 1 1>}
  NamedRerouteUsage()
  NamedRerouteUsage()
  NamedRerouteUsage()
  Power(Base=@n56) {constExponent=1.25}
  Saturate(None=@n39)
  n188 = Constant()
  SetMaterialAttributes!(
   MaterialAttributes=@n45
   BaseColor=Constant()
   Roughness=@n188
   Specular=@n188
   EmissiveColor=Multiply(A=NamedRerouteUsage())
   Metallic=Constant()
   Normal
  ) {attributeSetTypes [
    #2
    #3
    #5
    #9
    #4
    #1
   ]}
  n189 = VectorParameter() {group=Wrap_SkullPunk_ClothFuzz, parameterName=SKP_Fuzz Tint, defaultValue=<rgba 0.6823 0.6823 0.6823 1>}
  n190 = LinearInterpolate(A=@n153.BaseColor, B=Multiply(A=@n153.BaseColor, B=@n189.RGB), Alpha=@n162)
  ShadingPathSwitch(
   Default=BlendMaterialAttributes(
    A=NamedRerouteUsage()
    B=SetMaterialAttributes!(MaterialAttributes=@n153.MaterialAttributes, BaseColor=@n190) {attributeSetTypes [#2]}
    Alpha=NamedRerouteUsage()
   ) {vertexAttributeBlendType=UseA}
   Mobile=NamedRerouteUsage()
  )
  TextureSample(UVs=@n68.UVs) {texture=/Game/Weapons/WeaponSkins/Wraps/Textures/UVCheaker_Clamped}
 ]}
