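# Animated SlipStream wrap material function, blended over the input material by the primary mask behind a static switch.
# Projects scrolling backgrounds, streams and two asteroid layers using pre-skinned bounds and normals.
# Composites a sine-moving ship from a quadrant atlas, with warp, flickering thrusters and a beam that fades from a Blueprint-supplied timestamp.
# IsState inputs select scale, speed, rotation and offset per item type, such as weapon, vehicle or pickaxe.
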MaterialFunction /Game/Weapons/WeaponSkins/Wraps/MaterialFunctions/MF_Wrap_SlipStream
graph {nodes [
  n1 = NamedRerouteUsage()
  n2 = NamedRerouteUsage()
  n3 = MaterialFunctionCall(UVs=Add(A=Add(A=@n1, B=@n2), B=NamedRerouteUsage()) {constB=0}, Texture Scale=NamedRerouteUsage()) {materialFunction=&ScaleUVsByCenter}
  n4 = Multiply(A=@n3.UVs, B=Divide(B=Constant2Vector() {g=2, r=2}))
  n5 = TextureObjectParameter() {group=Wrap_SlipStream_ShipParts, parameterName=SLST_Ship_BC, texture=&T_BlacksRGB}
  n6 = TextureSample(UVs=Add(A=@n4, B=Constant2Vector()), Tex=@n5) {samplerSource=SSM_Clamp_WorldGroupSettings}
  n7 = NamedRerouteDeclaration(None=@n6.RGB) {name=Ship_BC, nodeColor=<rgba 0 0.5547 1 1>}
  n8 = TextureSample(UVs=Add(A=@n4, B=Constant2Vector() {r=0.5}), Tex=@n5) {samplerSource=SSM_Clamp_WorldGroupSettings}
  n9 = Add(A=@n4, B=Constant2Vector() {g=0.5})
  n10 = NamedRerouteUsage()
  n11 = Multiply(A=OneMinus(None=@n6.A), B=TextureSample(UVs=Add(A=@n9, B=@n10), Tex=@n5) {samplerSource=SSM_Clamp_WorldGroupSettings}.RGB)
  n12 = NamedRerouteUsage()
  n13 = Multiply(A=Multiply(A=OneMinus(None=@n8.A), B=Add(A=@n8.RGB, B=@n11)), B=@n12)
  n14 = Constant() {r=0.5}
  n15 = MaterialFunctionCall(Float2=@n3.UVs) {materialFunction=&BreakOutFloat2Components}
  n16 = SphereMask(A=@n15.G, B=@n14) {attenuationRadius=0.5}
  n17 = Multiply(A=SphereMask(A=@n14, B=@n15.R) {attenuationRadius=0.5}, B=@n16)
  Add(A=Multiply(A=Add(A=@n7, B=@n13), B=@n17)) {constB=0}
  n18 = NamedRerouteUsage()
  n19 = MaterialFunctionCall(Float4=NamedRerouteUsage()) {materialFunction=&BreakOutFloat4Components}
  n20 = NamedRerouteUsage()
  n21 = NamedRerouteDeclaration(None=MaterialFunctionCall(
    UVs=MaterialFunctionCall!(
     For_Safe_Zones
     Normal_Map
     Caculated_Bounds=@n18
     Pre_Skinned_Normals
     Is Vehicle=@n19.R
     Is VehicleSide=@n19.R
     Is Mech=@n19.R
     Vehicle_Rotation
     UV_Rotation Bypass=@n20
     Mech_Rotation
     Uvs_Offset
     Offset-after rotation
     Uvs_TileRate
    ) {materialFunction=&/MF_Projection_Side_Normal}.UVs No Offsets/Tiling
    Rotation Center=Constant2Vector() {g=0.5, r=0.5}
    "Rotation Angle (0-1)"=@n20
   ) {materialFunction=&CustomRotator}.Rotated Values) {name=Side_UVs, nodeColor=<rgba 0.9688 1 0 1>}
  n22 = Multiply(
   A=ComponentMask(None=VectorParameter() {group=Wrap_SlipStream_Astroid, parameterName=SLST_Asteroid_A_Speed, defaultValue=<rgba 0.15 0.013 0 1>}.RGB) {g=true, r=true}
   B=Time()
  )
  n23 = TextureObjectParameter() {group=Wrap_SlipStream_Astroid, parameterName=SLST_Asteroid__BC, texture=&T_BlacksRGB}
  n24 = TextureSample(
   UVs=Multiply(A=Add(A=@n21, B=@n22), B=ScalarParameter() {defaultValue=1, group=Wrap_SlipStream_Astroid, parameterName=SLST_Asteroid_A_Scale})
   Tex=@n23
  ) {samplerSource=SSM_Wrap_WorldGroupSettings}
  n25 = VectorParameter() {group=Wrap_SlipStream_Astroid, parameterName=SLST_Asteroid_B_Speed, defaultValue=<rgba 0.2 -0.0143 0 1>}
  n26 = Multiply(A=ComponentMask(None=@n25.RGB) {g=true, r=true}, B=Time())
  n27 = TextureSample(
   UVs=Multiply(A=Add(A=@n21, B=@n26), B=ScalarParameter() {defaultValue=2.5, group=Wrap_SlipStream_Astroid, parameterName=SLST_Asteroid_B_Scale})
   Tex=@n23
  ) {samplerSource=SSM_Wrap_WorldGroupSettings}
  Add(A=@n24.G, B=@n27.G)
  n28 = ScalarParameter() {defaultValue=0.25, group=Wrap_SlipStream_Astroid, parameterName=SLST_Stream_B_Brightness}
  n29 = MaterialFunctionCall(Float4=NamedRerouteUsage()) {materialFunction=&BreakOutFloat4Components}
  n30 = Multiply(A=NamedRerouteUsage(), B=NamedRerouteUsage())
  n31 = NamedRerouteUsage()
  n32 = ScalarParameter() {defaultValue=1, group=Wrap_SlipStream, parameterName=SLST_Back_Blend_Contrast}
  n33 = ScalarParameter() {defaultValue=1, group=Wrap_SlipStream, parameterName=SLST_Back_Fill_alternate_Axis_Opacity}
  n34 = VectorParameter() {group=Wrap_SlipStream, parameterName=SLST_Back_Fill_alternate_Axis}
  n35 = AppendVector(A=@n34.RGB, B=@n34.RGB)
  n36 = ComponentMask(None=MaterialFunctionCall!(
    IsCharacter=@n29.A
    IsStatsB=NamedRerouteUsage()
    Texture=NamedRerouteUsage()
    Rotation_Axis=NamedRerouteUsage()
    Caculated_Bounds=Add(A=@n30, B=@n31)
    Pre_Skinned_Normals=NamedRerouteUsage()
    Blend_Contrast=@n32
    Fill_alternate_Axis_Opacity=@n33
    "Fill_alternate_Axis_Color_A=Roughness"=@n35
    Is_Vehicle=@n29.A
    Is_Mech=@n29.A
    Enable_Fill_Z_Weapon
    Fill_Z_Weapon_opacity
    Fill_Z_Weapon_Color
   ) {materialFunction=&/MF_TopsSideProjection_Base}.RGB) {r=true}
  n37 = MaterialFunctionCall(Float4=NamedRerouteUsage()) {materialFunction=&BreakOutFloat4Components}
  n38 = NamedRerouteUsage()
  n39 = NamedRerouteUsage()
  n40 = MaterialFunctionCall(
   IsCharacter=@n37.A
   IsStatsB=NamedRerouteUsage()
   Texture=NamedRerouteDeclaration(None=TextureObjectParameter() {group=Wrap_SlipStream, parameterName=SLST_Back_Tex, texture=&T_BlacksRGB}) {name=Background tex, nodeColor=<rgba 1 0 0.9609 1>}
   Rotation_Axis=NamedRerouteUsage()
   Caculated_Bounds=Add(A=Multiply(A=@n38, B=@n39), B=NamedRerouteUsage())
   Pre_Skinned_Normals=NamedRerouteUsage()
   Blend_Contrast=@n32
   Fill_alternate_Axis_Opacity=@n33
   "Fill_alternate_Axis_Color_A=Roughness"=@n35
   Is_Vehicle=@n37.A
   Is_Mech=@n37.A
  ) {materialFunction=&MF_TopsSideProjection_Base}
  n41 = OneMinus(None=NamedRerouteUsage())
  n42 = NamedRerouteUsage()
  Add(A=Add(A=Multiply(A=@n28, B=@n36), B=Multiply(A=@n40.RGB, B=@n41)), B=@n42)
  n43 = NamedRerouteUsage()
  n44 = VectorParameter() {group=Wrap_SlipStream_UVs, parameterName=SLST_Streams_Back_Scale, defaultValue=<rgba 1 1 0 1>}
  n45 = ComponentMask(None=Multiply(A=@n44.RGB, B=@n44.RGB)) {g=true, r=true}
  n46 = ComponentMask(None=VectorParameter() {group=Wrap_SlipStream_UVs, parameterName=SLST_Streams_Back_Speed, defaultValue=<rgba 1.5 -0.0233 0 1>}.RGB) {g=true, r=true}
  n47 = Time()
  n48 = TextureSample(UVs=Add(A=Multiply(A=@n43, B=@n45), B=Multiply(A=@n46, B=@n47)), Tex=NamedRerouteUsage())
  n49 = VectorParameter() {group=Wrap_SlipStream_UVs, parameterName=SLST_Streams_A_Scale, defaultValue=<rgba 0.75 1 0 2>}
  n50 = ComponentMask(None=Multiply(A=@n49.RGB, B=@n49.RGB)) {g=true, r=true}
  n51 = ComponentMask(None=VectorParameter() {group=Wrap_SlipStream_UVs, parameterName=SLST_Streams_A_Speed, defaultValue=<rgba 0.5 0.023 0 1>}.RGB) {g=true, r=true}
  n52 = Time()
  n53 = TextureSample(UVs=Add(A=Multiply(A=@n43, B=@n50), B=Multiply(A=@n51, B=@n52)), Tex=NamedRerouteUsage())
  Add(A=@n48.RGB, B=@n53.R)
  n54 = NamedRerouteUsage()
  n55 = LinearInterpolate(A=Constant3Vector() {constant=<rgba 0 0 0 1>}, B=Add(A=NamedRerouteUsage(), B=Multiply(A=@n54, B=@n41)), Alpha=NamedRerouteUsage())
  Add(A=@n42, B=@n55)
  n56 = VectorParameter() {desc=Panning Speed Red and Green, group=Wrap_SlipStream_ShipParts_Uvs, parameterName=SLST_Ship_Speed_Weapon, sortPriority=4}
  n57 = VectorParameter() {desc=Panning Speed Red and Green, group=Wrap_SlipStream_ShipParts_Uvs, parameterName=SLST_Ship_Speed_Vehicle, sortPriority=5}
  n58 = ComponentMask(None=NamedRerouteUsage()) {r=true}
  n59 = Multiply(A=Time(), B=ComponentMask(None=LinearInterpolate(A=@n56.RGB, B=@n57.RGB, Alpha=@n58)) {g=true, r=true})
  n60 = ComponentMask(None=@n59) {r=true}
  n61 = ScalarParameter() {group=Wrap_SlipStream_ShipParts_Uvs, parameterName=SLST_Ship_Speed_offset, sortPriority=7}
  n62 = Sine(None=AppendVector(A=Add(A=@n60, B=@n61), B=ComponentMask(None=@n59) {g=true}))
  AppendVector(
   A=LinearInterpolate(
    A=ScalarParameter() {group=Wrap_SlipStream_ShipParts_Uvs, parameterName=SLST_Ship_X_Min, sortPriority=8}
    B=ScalarParameter() {defaultValue=1, group=Wrap_SlipStream_ShipParts_Uvs, parameterName=SLST_Ship_X_Max, sortPriority=9}
    Alpha=ComponentMask(None=@n62) {r=true}
   )
   B=LinearInterpolate(
    A=ScalarParameter() {group=Wrap_SlipStream_ShipParts_Uvs, parameterName=SLST_Ship_Y_Min, sortPriority=11}
    B=ScalarParameter() {defaultValue=1, group=Wrap_SlipStream_ShipParts_Uvs, parameterName=SLST_Ship_Y_Max, sortPriority=10}
    Alpha=ComponentMask(None=@n62) {g=true}
   )
  )
  n63 = TextureSample() {samplerType=SAMPLERTYPE_Masks, texture=/Game/Weapons/FORT_Rifles/Textures/T_Scar_CM}
  n64 = FunctionInput(Preview=@n63.R) {bUsePreviewValueAsDefault=true, inputName=Mask Primary, inputType=FunctionInput_Scalar, sortPriority=24, previewValue=<1 0 0 1>}
  n65 = BlendMaterialAttributes(A=NamedRerouteUsage(), B=NamedRerouteUsage(), Alpha=@n64) {vertexAttributeBlendType=UseA}
  n66 = FunctionInput(Preview=@n63.B) {bUsePreviewValueAsDefault=true, inputName=Mask Secondary, inputType=FunctionInput_Scalar, sortPriority=24, previewValue=<1 0 0 1>}
  BlendMaterialAttributes(A=@n65, Alpha=@n66) {vertexAttributeBlendType=UseA}
  n67 = VectorParameter() {group=Wrap_SlipStream_UVs, parameterName=SLST_Streams_A_Scale, defaultValue=<rgba 0.75 1 0 2>}
  ComponentMask(None=Multiply(A=@n67.RGB, B=@n67.RGB)) {g=true, r=true}
  Constant4Vector()
  Constant()
  FunctionInput(Preview=@n63.G) {bUsePreviewValueAsDefault=true, inputName=Mask Crevise, inputType=FunctionInput_Scalar, sortPriority=24, previewValue=<1 0 0 1>}
  n68 = TextureSample() {samplerType=SAMPLERTYPE_Normal, texture=/Game/Weapons/FORT_Rifles/T_SCAR_N}
  n69 = Constant3Vector() {constant=<rgba 1 0 0 0>}
  n70 = GetMaterialAttributes(None=FunctionInput(Preview=SetMaterialAttributes!(MaterialAttributes, Normal=@n68.RGB, BaseColor=@n69) {attributeSetTypes [#1, #2]}) {inputName=Material, inputType=FunctionInput_MaterialAttributes}) {attributeGetTypes [
    #2
    #3
    #1
    #4
   ]}
  FunctionOutput(None=StaticSwitchParameter(True=@n65, False=NamedRerouteDeclaration(None=@n70.MaterialAttributes) {name=Root_Material, nodeColor=<rgba 0.7812 1 0 1>}) {group=Customization_Wrap, parameterName=Wrap_SlipStream})
  n71 = Multiply(A=Multiply(A=@n8.RGB, B=@n8.RGB)) {constB=0.25}
  n72 = TextureSample(UVs=Add(A=@n4, B=Constant2Vector() {g=0.5, r=0.5}), Tex=@n5) {samplerSource=SSM_Clamp_WorldGroupSettings}
  n73 = Add(A=@n15.R)
  n74 = Multiply(A=@n16, B=OneMinus(None=SphereMask(A=@n73, B=@n14) {attenuationRadius=0.5}))
  n75 = NamedRerouteUsage()
  n76 = LinearInterpolate(A=Multiply(A=Add(A=@n71, B=@n72.RGB), B=@n74), Alpha=@n75) {constB=0}
  n77 = Multiply(
   A=Multiply(
    A=Multiply(A=@n8.RGB, B=@n74)
    B=ScalarParameter() {defaultValue=30, group=Wrap_SlipStream_ShipParts, parameterName="SLST_Ship_Beam\\Asteroid Rightness"}
   )
   B=NamedRerouteUsage()
  )
  n78 = Subtract(A=Time(), B=ScalarParameter() {defaultValue=-9999999, group=Wrap_TimefromBP, parameterName=TimeFrom-BP----------------, sortPriority=600})
  n79 = ScalarParameter() {defaultValue=0.7, group=Wrap_TimefromBP, parameterName=SLST_Ship_Beam_Time_Duration, sortPriority=601}
  n80 = Power(
   Base=Saturate(None=Divide(A=@n78, B=@n79))
   Exp=ScalarParameter() {defaultValue=0.5, group=Wrap_TimefromBP, parameterName=SLST_Ship_Beam_Time_Curve, sortPriority=601}
  )
  n81 = Multiply(A=@n12, B=Multiply(A=Add(A=@n76, B=@n77), B=OneMinus(None=@n80)))
  Multiply(A=@n81)
  n82 = VectorParameter() {group=Wrap_SlipStream_UVs_Back, parameterName=SLST_Rotation_Weapon, defaultValue=<rgba 0 0 -0.25 1>}
  n83 = VectorParameter() {
   desc=must activate TS_Vehicle_Side_01 UVs
   group=Wrap_SlipStream_UVs_Back
   parameterName=SLST_Rotation_Vehicle_Side
   defaultValue=<rgba 0 0 0 1>
  }
  n84 = MaterialFunctionCall(Float4=NamedRerouteUsage()) {materialFunction=&BreakOutFloat4Components}
  n85 = LinearInterpolate(
   A=VectorParameter() {group=Wrap_SlipStream_UVs_Back, parameterName=SLST_Rotation_Vehicle, defaultValue=<rgba 0 0 0 1>}.RGB
   B=@n83.RGB
   Alpha=@n84.G
  )
  n86 = VectorParameter() {group=Wrap_SlipStream_UVs_Back, parameterName=SLST_Rotation_Mech, defaultValue=<rgba 0.25 0.25 0 1>}
  n87 = VectorParameter() {group=Wrap_SlipStream_UVs_Back, parameterName=SLST_Rotation_Characte, defaultValue=<rgba 0.25 0.25 0 1>}
  NamedRerouteDeclaration(None=LinearInterpolate(A=LinearInterpolate(A=@n82.RGB, B=@n85, Alpha=@n84.R), B=LinearInterpolate(A=@n86.RGB, B=@n87.RGB, Alpha=@n84.A), Alpha=@n84.B)) {name=TopSide_Rotation, nodeColor=<rgba 0.9922 1 0 1>}
  n88 = VectorParameter() {group=Wrap_SlipStream_UVs_Back, parameterName=SLST_BackSpeed_Weapon_01, defaultValue=<rgba 0 0.1 0.023 1>}
  n89 = VectorParameter() {group=Wrap_SlipStream_UVs_Back, parameterName=SLST_BackSpeed_Vehicle_01, defaultValue=<rgba 0 0 0 1>}
  n90 = VectorParameter() {group=Wrap_SlipStream_UVs_Back, parameterName=SLST_BackSpeed_VehicleSide_01, defaultValue=<rgba 0 0 0 1>}
  n91 = MaterialFunctionCall(Float4=NamedRerouteUsage()) {materialFunction=&BreakOutFloat4Components}
  n92 = LinearInterpolate(
   A=LinearInterpolate(A=@n88.RGB, B=LinearInterpolate(A=@n89.RGB, B=@n90.RGB, Alpha=@n91.G), Alpha=@n91.R)
   B=LinearInterpolate(
    A=VectorParameter() {group=Wrap_SlipStream_UVs_Back, parameterName=SLST_BackSpeed_Mech_01, defaultValue=<rgba 0 0 0 1>}.RGB
    B=VectorParameter() {group=Wrap_SlipStream_UVs_Back, parameterName=SLST_BackSpeed_Character_01, defaultValue=<rgba 0 0 0 1>}.RGB
    Alpha=@n91.A
   )
   Alpha=@n91.B
  )
  n93 = VectorParameter() {group=Wrap_SlipStream_UVs_Back, parameterName=SLST_BackSpeed_Pickaxe_01, defaultValue=<rgba 0 0 0 1>}
  n94 = MaterialFunctionCall(Float2=NamedRerouteUsage()) {materialFunction=&BreakOutFloat2Components}
  NamedRerouteDeclaration(None=Multiply(A=LinearInterpolate(A=@n92, B=@n93.RGB, Alpha=@n94.G), B=Time())) {name=BackGground Speed, nodeColor=<rgba 0 0.625 1 1>}
  n95 = VectorParameter() {group=Wrap_SlipStream_ShipParts_Uvs, parameterName=SLST_Ship_X/Y Amount_Weapon, defaultValue=<rgba 0 0 0 1>}
  n96 = VectorParameter() {group=Wrap_SlipStream_ShipParts_Uvs, parameterName=SLST_Ship_X/Y Amount_vehicle, defaultValue=<rgba 0 0 0 1>}
  n97 = Multiply(A=@n62, B=LinearInterpolate(A=ComponentMask(None=@n95.RGB) {g=true, r=true}, B=ComponentMask(None=@n96.RGB) {g=true, r=true}, Alpha=@n58))
  NamedRerouteDeclaration(None=Multiply(A=Add(A=@n97)) {constB=0.5}) {name=Ship_UVs_Movement, nodeColor=<rgba 0 1 0.6484 1>}
  n98 = VectorParameter() {group=Wrap_SlipStream_UVs, parameterName=SLST_SLines_Speed_Weapon_01, defaultValue=<rgba 0 0.1 0.023 1>}
  n99 = MaterialFunctionCall(Float4=NamedRerouteUsage()) {materialFunction=&BreakOutFloat4Components}
  n100 = LinearInterpolate(
   A=VectorParameter() {group=Wrap_SlipStream_UVs, parameterName=SLST_SLines_Speed_Vehicle_01, defaultValue=<rgba 0 0 0 1>}.RGB
   B=VectorParameter() {group=Wrap_SlipStream_UVs, parameterName=SLST_SLines_Speed_VehicleSide_01, defaultValue=<rgba 0 0 0 1>}.RGB
   Alpha=@n99.G
  )
  n101 = VectorParameter() {group=Wrap_SlipStream_UVs, parameterName=SLST_SLines_Speed_Mech_01, defaultValue=<rgba 0 0 0 1>}
  n102 = VectorParameter() {group=Wrap_SlipStream_UVs, parameterName=SLST_SLines_Speed_Character_01, defaultValue=<rgba 0 0 0 1>}
  n103 = VectorParameter() {group=Wrap_SlipStream_UVs, parameterName=SLST_SLines_Speed_Backpack_01-----------, defaultValue=<rgba 0 0 0 1>}
  n104 = MaterialFunctionCall(Float2=NamedRerouteUsage()) {materialFunction=&BreakOutFloat2Components}
  n105 = LinearInterpolate(
   A=LinearInterpolate(A=LinearInterpolate(A=@n98.RGB, B=@n100, Alpha=@n99.R), B=LinearInterpolate(A=@n101.RGB, B=@n102.RGB, Alpha=@n99.A), Alpha=@n99.B)
   B=@n103.RGB
   Alpha=@n104.R
  )
  n106 = VectorParameter() {group=Wrap_SlipStream_UVs, parameterName=SLST_SLines_Speed_Pickaxe_01, defaultValue=<rgba 0 0 0 1>}
  NamedRerouteDeclaration(None=Multiply(A=LinearInterpolate(A=@n105, B=@n106.RGB, Alpha=@n104.G), B=Time())) {name=SpeedLines_Speed, nodeColor=<rgba 0 0.625 1 1>}
  n107 = ScalarParameter() {group=Wrap_SlipStream_UVs, parameterName=SLST_Side_Rotation_Weapon}
  n108 = MaterialFunctionCall(Float4=NamedRerouteUsage()) {materialFunction=&BreakOutFloat4Components}
  n109 = LinearInterpolate(
   A=ScalarParameter() {group=Wrap_SlipStream_UVs, parameterName=SLST_Side_Rotation_Vehicle}
   B=ScalarParameter() {group=Wrap_SlipStream_UVs, parameterName=SLST_Side_Rotation_VehicleSide}
   Alpha=@n108.G
  )
  n110 = ScalarParameter() {defaultValue=0.25, group=Wrap_SlipStream_UVs, parameterName=SLST_Side_Rotation_Mech}
  n111 = ScalarParameter() {group=Wrap_SlipStream_UVs, parameterName=SLST_Side_Rotation_Character}
  NamedRerouteDeclaration(None=LinearInterpolate(A=LinearInterpolate(A=@n107, B=@n109, Alpha=@n108.R), B=LinearInterpolate(A=@n110, B=@n111, Alpha=@n108.A), Alpha=@n108.B)) {name=Primary_Rotation, nodeColor=<rgba 1 0.7266 0 1>}
  n112 = NamedRerouteUsage()
  n113 = LinearInterpolate(
   A=Multiply(A=@n27.RGB, B=ScalarParameter() {defaultValue=0.25, group=Wrap_SlipStream_Astroid, parameterName=SLST_Asteroid_B_Brightness})
   B=NamedRerouteUsage()
   Alpha=@n112
  )
  n114 = NamedRerouteDeclaration(None=@n24.A) {name=Front_Asteroid, nodeColor=<rgba 0 0.6953 1 1>}
  n115 = Add(A=@n114, B=@n27.A)
  n116 = Multiply(A=Saturate(None=Add(A=@n115, B=@n112)), B=NamedRerouteUsage())
  NamedRerouteDeclaration(None=LinearInterpolate(
    A=VectorParameter() {group=Wrap_SlipStream, parameterName=SLST_BackGround_Color, defaultValue=<rgba 0 0 0 1>}.RGB
    B=LinearInterpolate(A=@n113, B=@n24.RGB, Alpha=@n114)
    Alpha=@n116
   )) {name=Asteroids_BC, nodeColor=<rgba 0 0.2031 1 1>}
  NamedRerouteDeclaration(None=@n116) {name=Asteroids_Mask, nodeColor=<rgba 0 0.2031 1 1>}
  n117 = Add(A=NamedRerouteUsage(), B=NamedRerouteUsage())
  n118 = NamedRerouteUsage()
  n119 = SetMaterialAttributes!(
   MaterialAttributes=NamedRerouteUsage()
   BaseColor=@n42
   Roughness=ScalarParameter() {group=Wrap_SlipStream, parameterName=SLST_Main_Roghness}
   Specular=ScalarParameter() {defaultValue=0.2, group=Wrap_SlipStream, parameterName=SLST_Main_Specular}
   Metallic=ScalarParameter() {group=Wrap_SlipStream, parameterName=SLST_Main_Metalic}
   Normal
   EmissiveColor=Add(A=@n55, B=Multiply(A=@n117, B=@n118))
  ) {attributeSetTypes [
    #2
    #3
    #5
    #4
    #1
    #6
   ]}
  NamedRerouteDeclaration(None=@n119) {name=Material_Back, nodeColor=<rgba 0 0.3906 1 1>}
  n120 = Saturate(None=@n115)
  n121 = OneMinus(None=@n112)
  n122 = Power(Base=LinearInterpolate(A=Multiply(A=@n121, B=@n27.G), B=@n24.G, Alpha=@n24.G))
  NamedRerouteDeclaration(None=Saturate(None=Multiply(A=@n120, B=@n122))) {name=Asteroid_lighting, nodeColor=<rgba 0 1 0.1562 1>}
  NamedRerouteDeclaration(None=Multiply(A=@n53.R, B=ScalarParameter() {defaultValue=0.25, group=Wrap_SlipStream_Astroid, parameterName=SLST_Stream_B_Brightness})) {name=Side_Streams, nodeColor=<rgba 0.3828 1 0 1>}
  n123 = MaterialFunctionCall(Color=Abs(None=NamedRerouteUsage())) {materialFunction=&SplitComponents}
  n124 = MaterialFunctionCall(Float4=NamedRerouteUsage()) {materialFunction=&BreakOutFloat4Components}
  n125 = LinearInterpolate(A=@n123.R, B=LinearInterpolate(A=@n123.B, B=@n123.B, Alpha=@n124.G), Alpha=@n124.R)
  NamedRerouteDeclaration(None=MaterialFunctionCall(
    In=LinearInterpolate(A=@n125, B=@n123.G, Alpha=@n124.B)
    Contrast=ScalarParameter() {defaultValue=1, group=Wrap_SlipStream, parameterName=SLST_Back_Blend_Contrast}
   ) {materialFunction=&CheapContrast}.Result) {name=Tri_Mask, nodeColor=<rgba 1 0.4688 0 1>}
  n126 = VectorParameter() {group=Wrap_SlipStream_UVs, parameterName=SLST_SLines_Scale_Weapon_01, defaultValue=<rgba 1 1 1 4>}
  n127 = VectorParameter() {group=Wrap_SlipStream_UVs, parameterName=SLST_SLines_Scale_Vehicle_01, defaultValue=<rgba 1 1 1 1>}
  n128 = Multiply(A=@n127.RGB, B=@n127.RGB)
  n129 = VectorParameter() {group=Wrap_SlipStream_UVs, parameterName=SLST_SLines_Scale_VehicleSide_01, defaultValue=<rgba 1 1 1 1>}
  n130 = Multiply(A=@n129.RGB, B=@n129.RGB)
  n131 = MaterialFunctionCall(Float4=NamedRerouteUsage()) {materialFunction=&BreakOutFloat4Components}
  n132 = VectorParameter() {group=Wrap_SlipStream_UVs, parameterName=SLST_SLines_Scale_Mech_01, defaultValue=<rgba 1 1 1 1>}
  n133 = VectorParameter() {group=Wrap_SlipStream_UVs, parameterName=SLST_SLines_Scale_Character_01, defaultValue=<rgba 1 1 1 1>}
  n134 = LinearInterpolate(
   A=LinearInterpolate(A=Multiply(A=@n126.RGB, B=@n126.RGB), B=LinearInterpolate(A=@n128, B=@n130, Alpha=@n131.G), Alpha=@n131.R)
   B=LinearInterpolate(A=Multiply(A=@n132.RGB, B=@n132.RGB), B=Multiply(A=@n133.RGB, B=@n133.RGB), Alpha=@n131.A)
   Alpha=@n131.B
  )
  n135 = VectorParameter() {group=Wrap_SlipStream_UVs, parameterName=SLST_SLines_Scale_Backpack_01-----------, defaultValue=<rgba 1 1 1 1>}
  n136 = Multiply(A=@n135.RGB, B=@n135.RGB)
  n137 = MaterialFunctionCall(Float2=NamedRerouteUsage()) {materialFunction=&BreakOutFloat2Components}
  n138 = VectorParameter() {group=Wrap_SlipStream_UVs, parameterName=SLST_SLines_Scale_Pickaxe_01, defaultValue=<rgba 1 1 1 1>}
  NamedRerouteDeclaration(None=LinearInterpolate(A=LinearInterpolate(A=@n134, B=@n136, Alpha=@n137.R), B=Multiply(A=@n138.RGB, B=@n138.RGB), Alpha=@n137.G)) {name=SpeedLines_Scale, nodeColor=<rgba 0.4062 0 1 1>}
  NamedRerouteDeclaration(None=Multiply(A=@n6.A, B=@n17)) {name=Ship_A, nodeColor=<rgba 0 0.5547 1 1>}
  n139 = VectorParameter() {group=Wrap_SlipStream_ShipParts_Uvs, parameterName=SLST_Ship_Offset_Weapon_01, defaultValue=<rgba -0.5 0.5 0 1>}
  n140 = VectorParameter() {group=Wrap_SlipStream_ShipParts_Uvs, parameterName=SLST_Ship_Offset_Vehicle_01, defaultValue=<rgba 0 0 0 1>}
  n141 = VectorParameter() {group=Wrap_SlipStream_ShipParts_Uvs, parameterName=SLST_Ship_Offset_VehicleSide_01, defaultValue=<rgba 0 0 0 1>}
  n142 = MaterialFunctionCall(Float4=NamedRerouteUsage()) {materialFunction=&BreakOutFloat4Components}
  n143 = VectorParameter() {group=Wrap_SlipStream_ShipParts_Uvs, parameterName=SLST_Ship_Offset_Mech_01, defaultValue=<rgba 0 0 0 1>}
  n144 = VectorParameter() {group=Wrap_SlipStream_ShipParts_Uvs, parameterName=SLST_Ship_Offset_Character_01, defaultValue=<rgba 0 0 0 1>}
  n145 = LinearInterpolate(
   A=LinearInterpolate(A=@n139.RGB, B=LinearInterpolate(A=@n140.RGB, B=@n141.RGB, Alpha=@n142.G), Alpha=@n142.R)
   B=LinearInterpolate(A=@n143.RGB, B=@n144.RGB, Alpha=@n142.A)
   Alpha=@n142.B
  )
  n146 = VectorParameter() {group=Wrap_SlipStream_ShipParts_Uvs, parameterName=SLST_Ship_Offset_Pickaxe_01, defaultValue=<rgba 0 0 0 1>}
  n147 = MaterialFunctionCall(Float2=NamedRerouteUsage()) {materialFunction=&BreakOutFloat2Components}
  NamedRerouteDeclaration(None=ComponentMask(None=LinearInterpolate(A=@n145, B=@n146.RGB, Alpha=@n147.G)) {g=true, r=true}) {name=Ship_offset, nodeColor=<rgba 0 0.625 1 1>}
  n148 = ScalarParameter() {defaultValue=0.25, group=Wrap_SlipStream_ShipParts_Uvs, parameterName=SLST_Ship_Scale_Weapon}
  n149 = MaterialFunctionCall(Float4=NamedRerouteUsage()) {materialFunction=&BreakOutFloat4Components}
  n150 = LinearInterpolate(
   A=ScalarParameter() {defaultValue=0.25, group=Wrap_SlipStream_ShipParts_Uvs, parameterName=SLST_Ship_Scale_Vehicle}
   B=ScalarParameter() {defaultValue=0.25, group=Wrap_SlipStream_ShipParts_Uvs, parameterName=SLST_Ship_Scale_VehicleSide}
   Alpha=@n149.G
  )
  n151 = ScalarParameter() {defaultValue=0.25, group=Wrap_SlipStream_ShipParts_Uvs, parameterName=SLST_Ship_Scale_Mech}
  n152 = ScalarParameter() {defaultValue=0.25, group=Wrap_SlipStream_ShipParts_Uvs, parameterName=SLST_Ship_Scale_Character}
  NamedRerouteDeclaration(None=LinearInterpolate(A=LinearInterpolate(A=@n148, B=@n150, Alpha=@n149.R), B=LinearInterpolate(A=@n151, B=@n152, Alpha=@n149.A), Alpha=@n149.B)) {name=Ship_Scale, nodeColor=<rgba 1 0.7266 0 1>}
  NamedRerouteDeclaration(None=@n81) {name=Beam_fire, nodeColor=<rgba 1 0 0.1172 1>}
  n153 = Add(
   A=Multiply(A=@n22, B=ScalarParameter() {defaultValue=10, group=Wrap_SlipStream_ShipParts, parameterName=SLST_ShipWarp_Speed_Mulp}) {constB=10}
   B=Multiply(A=@n3.UVs) {constB=2.3}
  )
  n154 = TextureObjectParameter() {group=Wrap_SlipStream_Astroid, parameterName=SLST_Stream__Msk, texture=&T_BlacksRGB}
  NamedRerouteDeclaration(None=Multiply(
    A=TextureSample(UVs=@n153, Tex=@n154) {samplerSource=SSM_Wrap_WorldGroupSettings}.A
    B=ScalarParameter() {defaultValue=0.02, group=Wrap_SlipStream_ShipParts, parameterName=SLST_ShipWarp_Amount}
   )) {name=Ship_Warp, nodeColor=<rgba 1 0 0.2109 1>}
  NamedRerouteDeclaration(None=@n154) {name=Stream_Mask, nodeColor=<rgba 1 0 0.4219 1>}
  n155 = Add(
   A=Floor(None=NamedRerouteUsage())
   B=Multiply(A=Time(), B=ScalarParameter() {defaultValue=0.992, group=Wrap_SlipStream_ShipParts, parameterName=SLST_Thruster_Flicker_Speed})
  )
  NamedRerouteDeclaration(None=LinearInterpolate(Alpha=TextureSample(UVs=@n155, Tex=@n154) {samplerSource=SSM_Wrap_WorldGroupSettings}.A) {constA=1, constB=5}) {name=Flicker_Speed, nodeColor=<rgba 1 0 0.1406 1>}
  NamedRerouteDeclaration(None=LinearInterpolate(A=Multiply(A=@n13, B=@n17), Alpha=@n75) {constA=1, constB=0}) {name=Thrusters, nodeColor=<rgba 1 0.9609 0 1>}
  NamedRerouteDeclaration(None=Multiply(A=ScalarParameter() {defaultValue=1, group=Wrap_SlipStream_Astroid, parameterName=SLST_Stream_A_Brightness}, B=@n48.RGB)) {name=Side_Back, nodeColor=<rgba 0 1 0.1328 1>}
  NamedRerouteDeclaration(None=FunctionInput(Preview=Constant4Vector().RGBA) {bUsePreviewValueAsDefault=true, inputName=IsState, inputType=FunctionInput_Vector4, sortPriority=4}) {name=IsState, nodeColor=<rgba 0.1497 1 0.1633 1>}
  NamedRerouteDeclaration(None=FunctionInput(Preview=Constant2Vector()) {bUsePreviewValueAsDefault=true, inputName=IsState_B, inputType=FunctionInput_Vector2, sortPriority=4}) {name=IsState_B, nodeColor=<rgba 0.1497 1 0.1633 1>}
  n156 = PreSkinnedLocalBounds()
  n157 = Subtract(A=VertexInterpolator(VS=PreSkinnedPosition()).PS, B=@n156.Min)
  n158 = MaterialFunctionCall(Color=Subtract(A=@n156.Max, B=@n156.Max)) {materialFunction=&SplitComponents}
  n159 = Max(A=Max(A=@n158.R, B=@n158.R), B=@n158.B)
  NamedRerouteDeclaration(None=FunctionInput(Preview=Divide(A=@n157, B=@n159)) {bUsePreviewValueAsDefault=true, inputName=Calcated_Bounds, sortPriority=11}) {name=Caclcated_Bounds, nodeColor=<rgba 0 0.9297 1 1>}
  n160 = ScalarParameter() {defaultValue=1, group=Wrap_SlipStream_UVs_Back, parameterName=SLST_ScaleBack_Vehicle}
  n161 = ScalarParameter() {defaultValue=1, group=Wrap_SlipStream_UVs_Back, parameterName=SLST_ScaleBack_Vehicle_Side}
  n162 = MaterialFunctionCall(Float4=NamedRerouteUsage()) {materialFunction=&BreakOutFloat4Components}
  n163 = LinearInterpolate(
   A=LinearInterpolate(
    A=ScalarParameter() {defaultValue=2.5, group=Wrap_SlipStream_UVs_Back, parameterName=SLST_ScaleBack_Weapon}
    B=LinearInterpolate(A=@n160, B=@n161, Alpha=@n162.G)
    Alpha=@n162.R
   )
   B=LinearInterpolate(
    A=ScalarParameter() {defaultValue=1, group=Wrap_SlipStream_UVs_Back, parameterName=SLST_ScaleBack_Mech}
    B=ScalarParameter() {defaultValue=1, group=Wrap_SlipStream_UVs_Back, parameterName=SLST_ScaleBack_Character}
    Alpha=@n162.A
   )
   Alpha=@n162.B
  )
  n164 = ScalarParameter() {defaultValue=1, group=Wrap_SlipStream_UVs_Back, parameterName=SLST_ScaleBack_Backpack}
  n165 = MaterialFunctionCall(Float2=NamedRerouteUsage()) {materialFunction=&BreakOutFloat2Components}
  NamedRerouteDeclaration(None=LinearInterpolate(
    A=LinearInterpolate(A=@n163, B=@n164, Alpha=@n165.R)
    B=ScalarParameter() {defaultValue=1, group=Wrap_SlipStream_UVs_Back, parameterName=SLST_ScaleBack_Pickaxe}
    Alpha=@n165.G
   )) {name=TopSide_Scale_Back, nodeColor=<rgba 0.4062 0 1 1>}
  n166 = PreSkinnedNormal()
  NamedRerouteDeclaration(None=FunctionInput(Preview=VertexInterpolator(VS=@n166).PS) {bUsePreviewValueAsDefault=true, inputName=Pre-Skinned Local Normal, sortPriority=10}) {name=PreSkin_Normal, nodeColor=<rgba 0.8281 1 0 1>}
  NamedRerouteDeclaration(None=@n70.Normal) {name=Root_Normal, nodeColor=<rgba 0.4062 0 1 1>}
  Saturate(None=@n64)
  SetMaterialAttributes!(MaterialAttributes=NamedRerouteUsage(), BaseColor=Constant(), Roughness, Specular, EmissiveColor, Metallic, Normal) {attributeSetTypes [
    #2
    #3
    #5
    #6
    #4
    #1
   ]}
 ]}
