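# Weapon wrap material function that adds a smashed-glass look with glowing cracks to weapons, vehicles, mechs and characters, behind a static switch.
# Side-projected textures, tuned per IsState, are resampled at four view-offset depths to fake layered cracks.
# Crack A is static, while B, C and D shrink and fade from Blueprint-set times.
# Cracks and decal drive roughness and time-of-day-scaled emissive, with crack normals added, blended through customization masks.
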MaterialFunction /Game/Weapons/WeaponSkins/Wraps/MaterialFunctions/MF_Wrap_SmashedGlass
graph {nodes [
  Constant() {r=1}
  FunctionInput(Preview=TextureSample() {samplerType=SAMPLERTYPE_Masks, texture=/Game/Weapons/FORT_Rifles/Textures/T_Scar_CM}.G) {bUsePreviewValueAsDefault=true, inputName=Curavture, inputType=FunctionInput_Scalar, sortPriority=24}
  n1 = FunctionInput(Preview=SetMaterialAttributes!(MaterialAttributes, Normal=TextureSample() {samplerType=SAMPLERTYPE_Normal, texture=/Game/Weapons/FORT_Rifles/T_SCAR_N}.RGB) {attributeSetTypes [#1]}) {bUsePreviewValueAsDefault=true, inputName=Material, inputType=FunctionInput_MaterialAttributes, sortPriority=-20}
  n2 = VectorParameter() {group=Wrap_SmashedGlass, parameterName=SmashedG_BaseColor, sortPriority=0, defaultValue=<rgba 0.003 0.003 0.003 1>}
  n3 = ScalarParameter() {defaultValue=0.2, group=Wrap_SmashedGlass, parameterName=SmashedG_Roghness_Low, sortPriority=0}
  n4 = ScalarParameter() {defaultValue=1, group=Wrap_SmashedGlass, parameterName=SmashedG_Roghness_High, sortPriority=0}
  n5 = TextureObjectParameter() {group=Wrap_SmashedGlass, parameterName="SmashedG__C\\DText_N", samplerType=SAMPLERTYPE_Normal, texture=/Engine/EngineMaterials/FlatNormal}
  n6 = FunctionInput(Preview=Constant()) {bUsePreviewValueAsDefault=true, inputName=IsState, inputType=FunctionInput_Vector4}
  n7 = MaterialFunctionCall(Color=@n6) {materialFunction=&SplitComponents}
  n8 = VertexInterpolator(VS=PreSkinnedPosition())
  n9 = PreSkinnedLocalBounds()
  n10 = FunctionInput(Preview=Divide(A=Subtract(A=@n8.PS, B=@n9.Min), B=Subtract(A=@n9.Max, B=@n9.Max))) {bUsePreviewValueAsDefault=true, inputName=Calcated_Bounds, sortPriority=11}
  n11 = ScalarParameter() {defaultValue=1, group=Wrap_SmashedGlass, parameterName=SmashedG_UV_TileRate_Vehicle}
  n12 = ScalarParameter() {defaultValue=1, group=Wrap_SmashedGlass, parameterName=SmashedG_UV_TileRate_VehicleSide}
  n13 = MaterialFunctionCall(Color=@n6) {materialFunction=&SplitComponents}
  n14 = LinearInterpolate(
   A=LinearInterpolate(
    A=ScalarParameter() {defaultValue=1, group=Wrap_SmashedGlass, parameterName=SmashedG_UV_TileRate_Weapon}
    B=LinearInterpolate(A=@n11, B=@n12, Alpha=@n13.G)
    Alpha=@n13.R
   )
   B=ScalarParameter() {defaultValue=1, group=Wrap_SmashedGlass, parameterName=SmashedG_UV_TileRate_Mech}
   Alpha=@n13.B
  )
  n15 = VectorParameter() {group=Wrap_SmashedGlass, parameterName=SmashedG_UV_Offset_Vehicle, defaultValue=<rgba 0 0 0 1>}
  n16 = VectorParameter() {group=Wrap_SmashedGlass, parameterName=SmashedG_UV_Offset_VehicleSide, defaultValue=<rgba 0 0 0 1>}
  n17 = MaterialFunctionCall(Float4=@n6) {materialFunction=&BreakOutFloat4Components}
  n18 = LinearInterpolate(
   A=LinearInterpolate(
    A=VectorParameter() {group=Wrap_SmashedGlass, parameterName=SmashedG_UV_Offset_Weapon, defaultValue=<rgba 0 0 0 1>}.RGB
    B=LinearInterpolate(A=@n15.RGB, B=@n16.RGB, Alpha=@n17.G)
    Alpha=@n17.R
   )
   B=LinearInterpolate(
    A=VectorParameter() {group=Wrap_SmashedGlass, parameterName=SmashedG_UV_Offset_Mech, defaultValue=<rgba 0 0 0 1>}.RGB
    B=VectorParameter() {group=Wrap_SmashedGlass, parameterName=SmashedG_UV_Offset_Character, defaultValue=<rgba 0 0 0 1>}.RGB
    Alpha=@n17.A
   )
   Alpha=@n17.B
  )
  n19 = ScalarParameter() {group=Wrap_SmashedGlass, parameterName=SmashedG_UV_Rotation_Vehicle}
  n20 = ScalarParameter() {defaultValue=0.25, group=Wrap_SmashedGlass, parameterName=SmashedG_UV_Rotation_Mech}
  n21 = FunctionInput(Preview=VertexInterpolator(VS=PreSkinnedNormal()).PS) {bUsePreviewValueAsDefault=true, inputName=Pre-Skinned Local Normal, sortPriority=10}
  n22 = MaterialFunctionCall!(
   For_Safe_Zones
   Normal_Map=@n5
   Caculated_Bounds=@n10
   Pre_Skinned_Normals=@n21
   Is Vehicle=@n7.R
   Is VehicleSide=@n7.R
   Is Mech=@n7.R
   Vehicle_Rotation=@n19
   UV_Rotation Bypass
   Mech_Rotation=@n20
   Uvs_Offset=@n18
   Offset-after rotation
   Uvs_TileRate=@n14
  ) {materialFunction=&/MF_Projection_Side_Normal}
  n23 = Add(
   A=ComponentMask(None=VectorParameter() {group=Wrap_SmashedGlass, parameterName=SmashedG_Crack_A__Postion, defaultValue=<rgba 0.09 0 0 1>}.RGB) {g=true, r=true}
   B=@n22.UVs
  )
  n24 = Constant() {r=0.5}
  n25 = ScalarParameter() {defaultValue=0.4, group=Wrap_SmashedGlass, parameterName=SmashedG_Crack_A_Radius, sliderMin=0.25}
  n26 = ScalarParameter() {defaultValue=-0.5, group=Wrap_SmashedGlass, parameterName=SmashedG_Crack_A_Hardness}
  n27 = TextureObjectParameter() {group=Wrap_SmashedGlass, parameterName="SmashedG__A\\B_Text", texture=&T_BlacksRGB}
  n28 = TextureSample(UVs=@n22.UVs, Tex=@n27)
  n29 = OneMinus(None=Saturate(None=Add(A=@n28.B, B=@n28.B)))
  n30 = MaterialFunctionCall() {materialFunction=/Engine/Functions/Engine_MaterialFunctions02/WorldPositionOffset/ObjectScale}
  n31 = MaterialFunctionCall(Color=@n6) {materialFunction=&SplitComponents}
  n32 = LinearInterpolate(
   A=ScalarParameter() {defaultValue=4, group=Wrap_SmashedGlass, parameterName=SmashedG_Layers_Depth_Amount}
   B=ScalarParameter() {defaultValue=1, group=Wrap_SmashedGlass, parameterName=SmashedG_Layers_Depth_Amount_Vehicle}
   Alpha=@n31.R
  )
  n33 = ScalarParameter() {defaultValue=1, group=Wrap_SmashedGlass, parameterName=SmashedG_Layers_Depth_Amount_Mech}
  n34 = Multiply(
   A=Transform(None=CameraVectorWS()) {transformSourceType=TRANSFORMSOURCE_World, transformType=TRANSFORM_Local}
   B=Multiply(A=Saturate(None=@n30.Scale Y), B=LinearInterpolate(A=@n32, B=@n33, Alpha=@n31.B))
  )
  n35 = MaterialFunctionCall(
   Caculated_Bounds=Add(A=@n10, B=Multiply(A=@n34) {constB=-0.001})
   Is Vehicle=@n7.R
   Is VehicleSide=@n7.R
   Is Mech=@n7.R
   Vehicle_Rotation=@n19
   Mech_Rotation=@n20
   Uvs_Offset=@n18
   Uvs_TileRate=@n14
  ) {materialFunction=&MF_Projection_Side_Normal}
  n36 = TextureSample(UVs=@n35.UVs, Tex=@n27)
  n37 = Abs(None=Subtract(A=@n28.G, B=@n36.R))
  n38 = MaterialFunctionCall(
   Caculated_Bounds=Add(A=@n10, B=Multiply(A=@n34) {constB=-0.002})
   Is Vehicle=@n7.R
   Is VehicleSide=@n7.R
   Is Mech=@n7.R
   Vehicle_Rotation=@n19
   Mech_Rotation=@n20
   Uvs_Offset=@n18
   Uvs_TileRate=@n14
  ) {materialFunction=&MF_Projection_Side_Normal}
  n39 = TextureSample(UVs=@n38.UVs, Tex=@n27)
  n40 = Abs(None=Subtract(A=@n28.G, B=@n39.R))
  n41 = MaterialFunctionCall(
   Caculated_Bounds=Add(A=@n10, B=Multiply(A=@n34) {constB=-0.003})
   Is Vehicle=@n7.R
   Is VehicleSide=@n7.R
   Is Mech=@n7.R
   Vehicle_Rotation=@n19
   Mech_Rotation=@n20
   Uvs_Offset=@n18
   Uvs_TileRate=@n14
  ) {materialFunction=&MF_Projection_Side_Normal}
  n42 = TextureSample(UVs=@n41.UVs, Tex=@n27)
  n43 = Abs(None=Subtract(A=@n28.R, B=@n42.R))
  n44 = MaterialFunctionCall(
   Caculated_Bounds=Add(A=@n10, B=Multiply(A=@n34) {constB=-0.004})
   Is Vehicle=@n7.R
   Is VehicleSide=@n7.R
   Is Mech=@n7.R
   Vehicle_Rotation=@n19
   Mech_Rotation=@n20
   Uvs_Offset=@n18
   Uvs_TileRate=@n14
  ) {materialFunction=&MF_Projection_Side_Normal}
  n45 = TextureSample(UVs=@n44.UVs, Tex=@n27)
  n46 = Abs(None=Subtract(A=@n28.R, B=@n45.R))
  n47 = Add(
   A=Add(A=Multiply(A=@n37) {constB=0.5}, B=Multiply(A=@n40) {constB=0.5})
   B=Add(A=Multiply(A=@n43) {constB=0.5}, B=Multiply(A=@n46) {constB=0.5}) {constB=0}
  )
  n48 = Multiply(A=MaterialFunctionCall(Alpha=Saturate(None=@n47)) {materialFunction=&SmoothStep}.Result) {constB=0.75}
  n49 = Multiply(A=Power(Base=SphereMask(A=@n23, B=@n24, Radius=@n25, Hardness=@n26)), B=Saturate(None=Add(A=@n29, B=@n48) {constB=0}))
  n50 = VectorParameter() {group=Wrap_SmashedGlass, parameterName=SmashedG_Crack_B__Postion, defaultValue=<rgba -0.34 -0.032 0 1>}
  n51 = Subtract(A=Time(), B=ScalarParameter() {defaultValue=-9999999, group=Wrap_SmashedGlass, parameterName=SmashedG_B_TimeFromeBP, sortPriority=90})
  n52 = ScalarParameter() {defaultValue=1.5, group=Wrap_SmashedGlass, parameterName=SmashedG_B_Crack_Duration, sortPriority=90}
  n53 = Power(
   Base=Saturate(None=Divide(A=@n51, B=@n52))
   Exp=ScalarParameter() {defaultValue=1.1, group=Wrap_SmashedGlass, parameterName=SmashedG_B_Crack_Falloff, sortPriority=90}
  ) {constExponent=0.95}
  n54 = SphereMask(
   A=Add(A=ComponentMask(None=@n50.RGB) {g=true, r=true}, B=@n22.UVs)
   B=Constant() {r=0.5}
   Radius=LinearInterpolate(A=ScalarParameter() {defaultValue=0.1, group=Wrap_SmashedGlass, parameterName=SmashedG_Crack_B_Radius, sliderMin=0.25}, Alpha=@n53) {constB=0}
   Hardness=LinearInterpolate(A=ScalarParameter() {defaultValue=-0.5, group=Wrap_SmashedGlass, parameterName=SmashedG_Crack_B_Hardness}, Alpha=@n53) {constB=-0.5}
  )
  n55 = Multiply(A=Saturate(None=Power(Base=@n54)), B=LinearInterpolate(Alpha=@n53) {constA=1, constB=0})
  n56 = OneMinus(None=@n28.A)
  n57 = Multiply(A=Abs(None=Subtract(A=@n28.G, B=@n36.G))) {constB=0.5}
  n58 = Multiply(A=Abs(None=Subtract(A=@n28.G, B=@n39.G))) {constB=0.5}
  n59 = Multiply(A=Abs(None=Subtract(A=@n28.G, B=@n42.G))) {constB=0.5}
  n60 = Multiply(A=Abs(None=Subtract(A=@n28.G, B=@n45.G))) {constB=0.5}
  n61 = Saturate(None=Add(A=Add(A=@n57, B=@n58), B=Add(A=@n59, B=@n60)))
  n62 = Add(
   A=Saturate(None=Multiply(A=@n28.B, B=@n56))
   B=Multiply(A=MaterialFunctionCall(Alpha=@n61) {materialFunction=&SmoothStep}.Result) {constB=0.75}
  )
  n63 = VectorParameter() {group=Wrap_SmashedGlass, parameterName="SmashedG_UV_Offset_C,D,Weapon", defaultValue=<rgba 0 0 0 1>}
  n64 = MaterialFunctionCall(Float4=@n6) {materialFunction=&BreakOutFloat4Components}
  n65 = LinearInterpolate(
   A=VectorParameter() {group=Wrap_SmashedGlass, parameterName="SmashedG_UV_Offset_C,D,Vehicle", defaultValue=<rgba 0 0 0 1>}.RGB
   B=VectorParameter() {group=Wrap_SmashedGlass, parameterName="SmashedG_UV_Offset_C,D,VehicleSide", defaultValue=<rgba 0 0 0 1>}.RGB
   Alpha=@n64.G
  )
  n66 = VectorParameter() {group=Wrap_SmashedGlass, parameterName="SmashedG_UV_Offset_C,D,Mech", defaultValue=<rgba 0 0 0 1>}
  n67 = VectorParameter() {group=Wrap_SmashedGlass, parameterName="SmashedG_UV_Offset_C,D,Character", defaultValue=<rgba 0 0 0 1>}
  n68 = ComponentMask(None=LinearInterpolate(A=LinearInterpolate(A=@n63.RGB, B=@n65, Alpha=@n64.R), B=LinearInterpolate(A=@n66.RGB, B=@n67.RGB, Alpha=@n64.A), Alpha=@n64.B)) {g=true, r=true}
  n69 = TextureObjectParameter() {
   group=Wrap_SmashedGlass
   parameterName="SmashedG__C\\D_Text"
   samplerType=SAMPLERTYPE_Masks
   texture=&T_BlackMask
  }
  n70 = TextureSample(UVs=Add(A=@n68, B=@n22.UVs), Tex=@n69)
  n71 = Saturate(None=Add(A=@n70.B, B=@n70.B))
  n72 = TextureSample(UVs=Add(A=@n35.UVs, B=@n68), Tex=@n69)
  n73 = Subtract(A=@n70.R, B=@n72.R)
  n74 = TextureSample(UVs=Add(A=@n38.UVs, B=@n68), Tex=@n69)
  n75 = Subtract(A=@n70.R, B=@n74.R)
  n76 = Add(A=Multiply(A=Abs(None=@n73)) {constB=0.5}, B=Multiply(A=Abs(None=@n75)) {constB=0.5})
  n77 = TextureSample(UVs=Add(A=@n41.UVs, B=@n68), Tex=@n69)
  n78 = Subtract(A=@n70.R, B=@n77.R)
  n79 = TextureSample(UVs=Add(A=@n44.UVs, B=@n68), Tex=@n69)
  n80 = Subtract(A=@n70.R, B=@n79.R)
  n81 = Add(A=Multiply(A=Abs(None=@n78)) {constB=0.5}, B=Multiply(A=Abs(None=@n80)) {constB=0.5})
  n82 = MaterialFunctionCall(Alpha=Saturate(None=Add(A=@n76, B=@n81))) {materialFunction=&SmoothStep}
  n83 = Saturate(None=Add(A=OneMinus(None=@n71), B=Multiply(A=@n82.Result) {constB=0.75}))
  n84 = VectorParameter() {group=Wrap_SmashedGlass, parameterName=SmashedG_Crack_C__Postion, defaultValue=<rgba -0.42 0.04 0 1>}
  n85 = Add(A=@n68, B=@n22.UVs)
  n86 = Subtract(A=Time(), B=ScalarParameter() {defaultValue=-9999999, group=Wrap_SmashedGlass, parameterName=SmashedG_C_TimeFromeBP, sortPriority=95})
  n87 = ScalarParameter() {defaultValue=1.5, group=Wrap_SmashedGlass, parameterName=SmashedG_C_Crack_Duration, sortPriority=95}
  n88 = Power(
   Base=Saturate(None=Divide(A=@n86, B=@n87))
   Exp=ScalarParameter() {defaultValue=1.1, group=Wrap_SmashedGlass, parameterName=SmashedG_C_Crack_Falloff, sortPriority=95}
  ) {constExponent=0.5}
  n89 = LinearInterpolate(A=ScalarParameter() {defaultValue=0.1, group=Wrap_SmashedGlass, parameterName=SmashedG_Crack_C_Radius, sliderMin=0.25}, Alpha=@n88) {constB=0}
  n90 = SphereMask(
   A=Add(A=ComponentMask(None=@n84.RGB) {g=true, r=true}, B=@n85)
   B=Constant() {r=0.5}
   Radius=@n89
   Hardness=LinearInterpolate(A=ScalarParameter() {defaultValue=-0.5, group=Wrap_SmashedGlass, parameterName=SmashedG_Crack_C_Hardness}, Alpha=@n88) {constB=-0.5}
  )
  n91 = Multiply(A=Saturate(None=Power(Base=@n90)), B=LinearInterpolate(Alpha=@n88) {constA=1, constB=0})
  n92 = Multiply(A=@n70.B, B=OneMinus(None=@n70.A))
  n93 = Subtract(A=@n70.G, B=@n72.G)
  n94 = Subtract(A=@n70.G, B=@n74.G)
  n95 = Add(A=Multiply(A=Abs(None=@n93)) {constB=0.5}, B=Multiply(A=Abs(None=@n94)) {constB=0.5})
  n96 = Subtract(A=@n70.G, B=@n77.G)
  n97 = Subtract(A=@n70.G, B=@n79.G)
  n98 = Add(A=Multiply(A=Abs(None=@n96)) {constB=0.5}, B=Multiply(A=Abs(None=@n97)) {constB=0.5})
  n99 = MaterialFunctionCall(Alpha=Saturate(None=Add(A=@n95, B=@n98))) {materialFunction=&SmoothStep}
  n100 = Saturate(None=Add(A=Saturate(None=@n92), B=Multiply(A=@n99.Result) {constB=0.75}))
  n101 = VectorParameter() {group=Wrap_SmashedGlass, parameterName=SmashedG_Crack_D__Postion, defaultValue=<rgba -0.25 0.06 0 1>}
  n102 = Subtract(A=Time(), B=ScalarParameter() {defaultValue=-9999999, group=Wrap_SmashedGlass, parameterName=SmashedG_D_TimeFromeBP, sortPriority=100})
  n103 = ScalarParameter() {defaultValue=1.5, group=Wrap_SmashedGlass, parameterName=SmashedG_D_Crack_Duration, sortPriority=100}
  n104 = Power(
   Base=Saturate(None=Divide(A=@n102, B=@n103))
   Exp=ScalarParameter() {defaultValue=1.1, group=Wrap_SmashedGlass, parameterName=SmashedG_D_Crack_Falloff, sortPriority=100}
  ) {constExponent=0.25}
  n105 = SphereMask(
   A=Add(A=ComponentMask(None=@n101.RGB) {g=true, r=true}, B=@n85)
   B=Constant() {r=0.5}
   Radius=LinearInterpolate(A=ScalarParameter() {defaultValue=0.1, group=Wrap_SmashedGlass, parameterName=SmashedG_Crack_D_Radius, sliderMin=0.25}, Alpha=@n104) {constB=0}
   Hardness=LinearInterpolate(A=ScalarParameter() {defaultValue=-0.5, group=Wrap_SmashedGlass, parameterName=SmashedG_Crack_D_Hardness}, Alpha=@n104) {constB=-0.5}
  )
  n106 = Multiply(A=Saturate(None=Power(Base=@n105)), B=LinearInterpolate(Alpha=@n104) {constA=1, constB=0})
  n107 = Add(A=Multiply(A=@n55, B=Saturate(None=@n62)), B=Add(A=Multiply(A=@n83, B=@n91), B=Multiply(A=@n100, B=@n106)))
  n108 = MaterialFunctionCall(Color=Abs(None=@n21)) {materialFunction=&SplitComponents}
  n109 = MaterialFunctionCall(Color=@n6) {materialFunction=&SplitComponents}
  n110 = LinearInterpolate(A=@n108.B, B=@n108.B, Alpha=@n109.G)
  n111 = ScalarParameter() {defaultValue=0.5, group=Wrap_SmashedGlass, parameterName=SmashedG_Tri_Planner_MaskContrast_Weapon, sortPriority=5}
  n112 = MaterialFunctionCall(Color=@n6) {materialFunction=&SplitComponents}
  n113 = LinearInterpolate(
   A=ScalarParameter() {defaultValue=0.5, group=Wrap_SmashedGlass, parameterName=SmashedG_Tri_Planner_MaskContrast_Vehicle, sortPriority=5}
   B=ScalarParameter() {defaultValue=0.5, group=Wrap_SmashedGlass, parameterName=SmashedG_Tri_Planner_MaskContrast_VehicleSide, sortPriority=5}
   Alpha=@n112.G
  )
  n114 = MaterialFunctionCall(
   In=LinearInterpolate(A=LinearInterpolate(A=@n108.R, B=@n110, Alpha=@n109.R), B=@n108.G, Alpha=@n109.B)
   Contrast=LinearInterpolate(
    A=LinearInterpolate(A=@n111, B=@n113, Alpha=@n112.R)
    B=ScalarParameter() {defaultValue=0.5, group=Wrap_SmashedGlass, parameterName=SmashedG_Tri_Planner_MaskContrast_Mech, sortPriority=5}
    Alpha=@n112.B
   )
  ) {materialFunction=&CheapContrast}
  n115 = Multiply(A=@n114.Result, B=Add(A=@n49, B=@n107))
  n116 = TextureSampleParameter2D(UVs=@n22.UVs) {
   group=Wrap_SmashedGlass
   parameterName=SmashedG_Decal_Text
   samplerType=SAMPLERTYPE_Masks
   texture=&T_BlackMask
  }
  n117 = ScalarParameter() {defaultValue=0.75, group=Wrap_SmashedGlass, parameterName=SmashedG_Emissive_Amount, sortPriority=0}
  n118 = VectorParameter() {group=Wrap_SmashedGlass, parameterName=SmashedG_CrackA__Color, defaultValue=<rgba 0.0497 0.9823 0.9823 1>}
  n119 = VectorParameter() {group=Wrap_SmashedGlass, parameterName="SmashedG_CrackB,C,D,__Color", defaultValue=<rgba 0.0497 0.9823 0.9823 1>}
  n120 = Add(A=Multiply(A=@n118.RGB, B=@n49), B=Multiply(A=@n119.RGB, B=@n107))
  n121 = Add(
   A=Multiply(A=VectorParameter() {group=Wrap_SmashedGlass, parameterName=SmashedG_Decal_Color, defaultValue=<rgba 0.4219 0.013 0 1>}.RGB, B=@n116.R)
   B=Multiply(A=Multiply(A=@n114.Result, B=@n120), B=OneMinus(None=@n116.G))
  )
  n122 = Constant() {r=0.5}
  n123 = Constant() {r=1}
  n124 = Constant() {r=0.5}
  n125 = Constant() {r=0.25}
  n126 = GetMaterialAttributes(None=@n1) {attributeGetTypes [#1]}
  n127 = TextureObjectParameter() {group=Wrap_SmashedGlass, parameterName="SmashedG__A\\B_Text_N", samplerType=SAMPLERTYPE_Normal, texture=&FlatNormal}
  n128 = MaterialFunctionCall(
   Normal_Map=@n127
   Caculated_Bounds=@n10
   Pre_Skinned_Normals=@n21
   Is Vehicle=@n7.R
   Is VehicleSide=@n7.R
   Is Mech=@n7.R
   Vehicle_Rotation=@n19
   Mech_Rotation=@n18
   Uvs_Offset=@n18
   Uvs_TileRate=@n14
  ) {materialFunction=&MF_Projection_Side_Normal}
  n129 = Multiply(A=@n114.Result, B=LinearInterpolate(A=@n128.Raw_Normal, B=Multiply(A=@n128.Raw_Normal, B=@n55), Alpha=@n28.B) {constB=0})
  n130 = LinearInterpolate(A=Multiply(A=@n22.Raw_Normal, B=@n91), B=Multiply(A=@n22.Raw_Normal, B=@n106), Alpha=@n70.B) {constB=0}
  n131 = SetMaterialAttributes!(
   MaterialAttributes=@n1
   BaseColor=@n2.RGB
   Roughness=LinearInterpolate(
    A=LinearInterpolate(A=@n3, B=@n4, Alpha=@n115)
    B=ScalarParameter() {defaultValue=0.3, group=Wrap_SmashedGlass, parameterName=SmashedG_Roghness_Decal, sortPriority=0}
    Alpha=@n116.R
   )
   Specular=ScalarParameter() {defaultValue=0.5, group=Wrap_SmashedGlass, parameterName=SmashedG_Specular, sortPriority=0}
   EmissiveColor=Multiply(
    A=Multiply(A=@n121, B=@n117)
    B=MaterialFunctionCall(Value_Morning=@n122, Value_Day=@n123, Value_Evening=@n124, Value_Night=@n125) {materialFunction=&Scalar_Time_of_Day}.Result
   )
   Metallic=ScalarParameter() {group=Wrap_SmashedGlass, parameterName=SmashedG_Metallic, sortPriority=0}
   Normal=Add(A=Add(A=@n126.Normal, B=@n129), B=Multiply(A=@n114.Result, B=@n130))
  ) {attributeSetTypes [
    #2
    #3
    #4
    #5
    #6
    #1
   ]}
  n132 = FunctionInput(Preview=TextureSample() {samplerType=SAMPLERTYPE_Masks, texture=&T_Scar_CM}.R) {bUsePreviewValueAsDefault=true, inputName="Customization Mask (_CM B)", inputType=FunctionInput_Scalar, sortPriority=24}
  n133 = ScalarParameter() {group=Wrap_SmashedGlass, parameterName=SmashedG_Remove_Customization Mask, sortPriority=-10}
  n134 = NamedRerouteUsage()
  n135 = NamedRerouteUsage()
  n136 = LinearInterpolate(
   A=Saturate(None=Add(A=@n132, B=@n133))
   B=FunctionInput() {bUsePreviewValueAsDefault=true, inputName=Customization Mask B, inputType=FunctionInput_Scalar, sortPriority=24}
   Alpha=Saturate(None=Subtract(A=@n134, B=@n135))
  )
  FunctionOutput(None=StaticSwitchParameter(True=BlendMaterialAttributes(A=@n1, B=@n131, Alpha=@n136) {vertexAttributeBlendType=UseA}, False=@n1) {group=Customization_Wrap, parameterName=Wrap_SmashedGlass})
  Multiply(A=@n91, B=@n89)
  NamedRerouteDeclaration(None=MaterialFunctionCall(Float4=@n6) {materialFunction=&BreakOutFloat4Components}.A) {name=isCharcter, nodeColor=<rgba 0.7792 1 0.0342 1>}
  NamedRerouteDeclaration(None=FunctionInput() {bUsePreviewValueAsDefault=true, inputName=isCharAccent, inputType=FunctionInput_Scalar, sortPriority=78}) {name=isCharAccent, nodeColor=<rgba 0 0.9253 1 1>}
  ScalarParameter() {defaultValue=0.2, group=Wrap_SmashedGlass, parameterName=SmashedG_Roghness, sortPriority=0}
 ]}
