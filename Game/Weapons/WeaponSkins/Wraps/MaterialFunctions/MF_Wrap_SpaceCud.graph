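# Material function for the Space Cuddles cosmetic wrap, blended over incoming material attributes by the customization mask behind a static switch.
# A side-projection function builds UVs and normals, with rotation, offset and tiling chosen by isState for weapon, vehicle, vehicle side, mech or character.
# Emissive masks shift in UV by rounded progression stage and accumulator values.
# Time-stamp parameters drive the color fade and ease both progression values back to zero.
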MaterialFunction /Game/Weapons/WeaponSkins/Wraps/MaterialFunctions/MF_Wrap_SpaceCud
graph {nodes [
  n1 = FunctionInput() {bUsePreviewValueAsDefault=true, inputName=Material, inputType=FunctionInput_MaterialAttributes, sortPriority=-20}
  n2 = GetMaterialAttributes(None=@n1) {attributeGetTypes [#1]}
  n3 = TextureObjectParameter() {
   group=Wrap_SpaceCuddles
   parameterName=SCud__Tex_Norm
   samplerType=SAMPLERTYPE_Normal
   texture=&FlatNormalBlue
  }
  n4 = PreSkinnedLocalBounds()
  n5 = Subtract(A=VertexInterpolator(VS=PreSkinnedPosition()).PS, B=@n4.Min)
  n6 = Subtract(A=@n4.Max, B=@n4.Max)
  n7 = PreSkinnedNormal()
  n8 = FunctionInput(Preview=Constant()) {inputName=isState, inputType=FunctionInput_Vector4, sortPriority=5}
  n9 = MaterialFunctionCall(Color=@n8) {materialFunction=&SplitComponents}
  n10 = ScalarParameter() {group=Wrap_SpaceCuddles, parameterName=SCud_UV_Rotation_Weapon}
  n11 = MaterialFunctionCall(Float4=@n8) {materialFunction=&BreakOutFloat4Components}
  n12 = LinearInterpolate(
   A=ScalarParameter() {group=Wrap_SpaceCuddles, parameterName=SCud_UV_Rotation_Vehicle}
   B=ScalarParameter() {group=Wrap_SpaceCuddles, parameterName=SCud_UV_Rotation_VehicleSide}
   Alpha=@n11.G
  )
  n13 = ScalarParameter() {group=Wrap_SpaceCuddles, parameterName=SCud_UV_Rotation_Mech}
  n14 = ScalarParameter() {group=Wrap_SpaceCuddles, parameterName=SCud_UV_Rotation_Character}
  n15 = VectorParameter() {group=Wrap_SpaceCuddles, parameterName=SCud_UV_Offset_Weapon, defaultValue=<rgba 0 0 0 1>}
  n16 = MaterialFunctionCall(Float4=@n8) {materialFunction=&BreakOutFloat4Components}
  n17 = LinearInterpolate(
   A=VectorParameter() {group=Wrap_SpaceCuddles, parameterName=SCud_UV_Offset_Vehicle, defaultValue=<rgba 0 0 0 1>}.RGB
   B=VectorParameter() {group=Wrap_SpaceCuddles, parameterName=SCud_UV_Offset_VehicleSide, defaultValue=<rgba 0 0 0 1>}.RGB
   Alpha=@n16.G
  )
  n18 = VectorParameter() {group=Wrap_SpaceCuddles, parameterName=SCud_UV_Offset_Mech, defaultValue=<rgba 0 0 0 1>}
  n19 = VectorParameter() {group=Wrap_SpaceCuddles, parameterName=SCud_UV_Offset_Character, defaultValue=<rgba 0 0 0 1>}
  n20 = ScalarParameter() {defaultValue=1, group=Wrap_SpaceCuddles, parameterName=SCud_UV_TileRate_Weapon}
  n21 = MaterialFunctionCall(Float4=@n8) {materialFunction=&BreakOutFloat4Components}
  n22 = LinearInterpolate(
   A=ScalarParameter() {defaultValue=1, group=Wrap_SpaceCuddles, parameterName=SCud_UV_TileRate_Vehicle}
   B=ScalarParameter() {defaultValue=1, group=Wrap_SpaceCuddles, parameterName=SCud_UV_TileRate_VehicleSide}
   Alpha=@n21.G
  )
  n23 = ScalarParameter() {defaultValue=1, group=Wrap_SpaceCuddles, parameterName=SCud_UV_TileRate_Mech}
  n24 = ScalarParameter() {defaultValue=1, group=Wrap_SpaceCuddles, parameterName=SCud_UV_TileRate_Character}
  n25 = MaterialFunctionCall!(
   For_Safe_Zones
   Normal_Map=@n3
   Caculated_Bounds=FunctionInput(Preview=Divide(A=@n5, B=@n6)) {bUsePreviewValueAsDefault=true, inputName=Calcated_Bounds, sortPriority=11}
   Pre_Skinned_Normals=FunctionInput(Preview=VertexInterpolator(VS=@n7).PS) {bUsePreviewValueAsDefault=true, inputName=Pre-Skinned Local Normal, sortPriority=10}
   Is Vehicle=@n9.R
   Is VehicleSide=@n9.R
   Is Mech=@n9.R
   Vehicle_Rotation
   UV_Rotation Bypass=LinearInterpolate(A=LinearInterpolate(A=@n10, B=@n12, Alpha=@n11.R), B=LinearInterpolate(A=@n13, B=@n14, Alpha=@n11.A), Alpha=@n11.B)
   Mech_Rotation
   Uvs_Offset=LinearInterpolate(A=LinearInterpolate(A=@n15.RGB, B=@n17, Alpha=@n16.R), B=LinearInterpolate(A=@n18.RGB, B=@n19.RGB, Alpha=@n16.A), Alpha=@n16.B)
   Offset-after rotation
   Uvs_TileRate=LinearInterpolate(A=LinearInterpolate(A=@n20, B=@n22, Alpha=@n21.R), B=LinearInterpolate(A=@n23, B=@n24, Alpha=@n21.A), Alpha=@n21.B)
  ) {materialFunction=&/MF_Projection_Side_Normal}
  n26 = TextureSample(
   UVs=@n25.UVs
   Tex=TextureObjectParameter() {group=Wrap_SpaceCuddles, parameterName=SCud__Tex_BC, texture=&T_BlacksRGB}
  ) {samplerType=SAMPLERTYPE_Normal}
  n27 = TextureObjectParameter() {group=Wrap_SpaceCuddles, parameterName=SCud__Tex_Msk, texture=&T_BlacksRGB}
  n28 = TextureSample(UVs=@n25.UVs, Tex=@n27) {samplerType=SAMPLERTYPE_Normal}
  n29 = Add(
   A=@n25.UVs
   B=ComponentMask(None=VectorParameter() {group=Wrap_SpaceCuddles, parameterName=SCud_Progression_Offset, defaultValue=<rgba -0.42 0 0 1>}.RGB) {g=true, r=true}
  )
  n30 = ScalarParameter() {defaultValue=0.12, group=Wrap_SpaceCuddles, parameterName=SCud_Progression_Amount_ScaleRange}
  n31 = Constant()
  n32 = ScalarParameter() {group=Wrap_SpaceCuddles, parameterName=SCud_Progression_Stage}
  n33 = Time()
  n34 = ScalarParameter() {defaultValue=9999999, group=Wrap_SpaceCuddles, parameterName=SCud_Time_BP_Reset, sortPriority=100}
  n35 = Saturate(None=Divide(
    A=Subtract(A=@n33, B=@n34)
    B=ScalarParameter() {defaultValue=0.25, group=Wrap_SpaceCuddles, parameterName=SCud_Time_BP_Reset_Duration, sortPriority=100}
   ))
  n36 = ScalarParameter() {defaultValue=4, group=Wrap_SpaceCuddles, parameterName=SCud_Progression_Stage_Amount}
  n37 = Divide(A=Round(None=LinearInterpolate(A=@n32, Alpha=@n35) {constB=0}), B=@n36) {constB=8}
  n38 = Add(A=@n29, B=Multiply(A=AppendVector(A=@n30, B=@n31), B=AppendVector(A=@n37, B=@n31)))
  n39 = Constant()
  n40 = ScalarParameter() {group=Wrap_SpaceCuddles, parameterName=SCud_Progression_Stage_Acumulator}
  n41 = Divide(A=Round(None=LinearInterpolate(A=@n40, Alpha=@n35) {constB=0}), B=@n36) {constB=8}
  n42 = Add(A=@n29, B=Multiply(A=AppendVector(A=@n30, B=@n39), B=AppendVector(A=@n41, B=@n39)))
  n43 = Add(
   A=Multiply(A=TextureSample(UVs=@n38, Tex=@n27) {samplerSource=SSM_Clamp_WorldGroupSettings, samplerType=SAMPLERTYPE_Normal}.A) {constB=0.5}
   B=Multiply(
    A=TextureSample(UVs=@n42, Tex=@n27) {samplerSource=SSM_Clamp_WorldGroupSettings, samplerType=SAMPLERTYPE_Normal}.A
    B=ScalarParameter() {defaultValue=1, group=Wrap_SpaceCuddles, parameterName=SCud_Progression_Stage_Acumulator_Brightness}
   ) {constB=0.5}
  )
  n44 = VectorParameter() {group=Wrap_SpaceCuddles, parameterName=SCud_Progression_Color_Start, defaultValue=<rgba 3 0.1 2.656 1>}
  n45 = VectorParameter() {group=Wrap_SpaceCuddles, parameterName=SCud_Progression_Color_End, defaultValue=<rgba 0.5 0.0167 0.4427 1>}
  n46 = Time()
  n47 = ScalarParameter() {defaultValue=-9999999, group=Wrap_SpaceCuddles, parameterName=SCud_Time_BP, sortPriority=100}
  n48 = Saturate(None=Divide(
    A=Subtract(A=@n46, B=@n47)
    B=ScalarParameter() {defaultValue=0.25, group=Wrap_SpaceCuddles, parameterName=SCud_Time_BP_Duration, sortPriority=100}
   ))
  n49 = SetMaterialAttributes!(
   MaterialAttributes=@n2.MaterialAttributes
   BaseColor=@n26.RGB
   Roughness=LinearInterpolate(
    A=ScalarParameter() {group=Wrap_SpaceCuddles, parameterName=SCud_Roghness_Low, sortPriority=0}
    B=ScalarParameter() {defaultValue=1, group=Wrap_SpaceCuddles, parameterName=SCud_Roghness_High, sortPriority=0}
    Alpha=@n26.A
   )
   Specular=ScalarParameter() {defaultValue=0.5, group=Wrap_SpaceCuddles, parameterName=SCud_Specular, sortPriority=0}
   EmissiveColor=Multiply(A=Multiply(A=@n28.R, B=@n43), B=LinearInterpolate(A=@n44.RGB, B=@n45.RGB, Alpha=@n48))
   Metallic=ScalarParameter() {group=Wrap_SpaceCuddles, parameterName=SCud_Metallic, sortPriority=0}
   Normal=LinearInterpolate(
    A=@n2.Normal
    B=Add(A=@n2.Normal, B=@n25.Raw_Normal)
    Alpha=ScalarParameter() {defaultValue=1, group=Wrap_SpaceCuddles, parameterName=SCud_Normal_Strength, sortPriority=0}
   )
  ) {attributeSetTypes [
    #2
    #3
    #4
    #5
    #6
    #1
   ]}
  n50 = Saturate(None=Add(
    A=FunctionInput() {bUsePreviewValueAsDefault=true, inputName="Customization Mask (_CM B)", inputType=FunctionInput_Scalar, sortPriority=24}
    B=ScalarParameter() {group=Wrap_SpaceCuddles, parameterName=SCud_Remove_Customization Mask, sortPriority=-10}
   ))
  FunctionOutput(None=StaticSwitchParameter(True=BlendMaterialAttributes(A=@n2.MaterialAttributes, B=@n49, Alpha=@n50) {vertexAttributeBlendType=UseA}, False=@n1) {group=Customization_Wrap, parameterName=Wrap_SpaceCuddles})
  Multiply()
 ]}
