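# Material function for the Squash weapon-skin wrap: an animated storm of clouds, rain and lightning, blended over the input material by customization mask behind a static switch.
# Safe-zone virtual-plane UVs drive two warped, panning cloud layers with a swirling core sized by object type.
# Cloud height distorts the two rain layers and masks lightning.
# Blueprint-fed game time triggers a fading lightning burst, after which randomized ambient flickers return, alongside Fresnel glow.
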MaterialFunction /Game/Weapons/WeaponSkins/Wraps/MaterialFunctions/MF_Wrap_Squash
graph {nodes [
  n1 = VertexInterpolator(VS=PreSkinnedPosition())
  n2 = PreSkinnedLocalBounds()
  FunctionInput(Preview=Divide(A=Subtract(A=@n1.PS, B=@n2.Min), B=Subtract(A=@n2.Max, B=@n2.Max))) {bUsePreviewValueAsDefault=true, inputName=Calcated_Bounds, sortPriority=11}
  FunctionInput(Preview=Constant3Vector() {constant=<rgba 0 1 0 0>}) {bUsePreviewValueAsDefault=true, inputName=ForwardAxis}
  FunctionInput(Preview=VertexInterpolator(VS=PreSkinnedNormal()).PS) {bUsePreviewValueAsDefault=true, inputName=Pre-Skinned Local Normal, sortPriority=10}
  n3 = FunctionInput(Preview=SetMaterialAttributes!(MaterialAttributes, EmissiveColor=TextureSample() {texture=/Game/Weapons/WeaponSkins/Wraps/Textures/UVCheaker}.RGB) {attributeSetTypes [#1]}) {inputName=Material, inputType=FunctionInput_MaterialAttributes, sortPriority=-10}
  n4 = GetMaterialAttributes(None=@n3) {attributeGetTypes [#2, #3]}
  n5 = Time()
  n6 = ComponentMask(None=VectorParameter() {group=Wrap_Squash, parameterName=SQ_Rain_Speed_A, defaultValue=<rgba 0.01 -1 0 1>}.RGB) {g=true, r=true}
  n7 = NamedRerouteUsage()
  n8 = MaterialFunctionCall(UVs=@n7, "Rotation Angle (0-1)"=ScalarParameter() {defaultValue=-0.05, group=Wrap_Squash, parameterName=SQ_Rain_Rotation}) {materialFunction=&CustomRotator}
  n9 = ScalarParameter() {defaultValue=1, group=Wrap_Squash, parameterName=SQ_Rain_Scale_A}
  n10 = ScalarParameter() {defaultValue=1, group=Wrap_Squash, parameterName=SQ_Master_tile_Size}
  n11 = ScalarParameter() {defaultValue=0.25, group=Wrap_Squash, parameterName=SQ_SwirlRadiusWeapon}
  n12 = ScalarParameter() {defaultValue=0.75, group=Wrap_Squash, parameterName=SQ_SwirlRadiusVehicle}
  n13 = FunctionInput(Preview=Constant()) {bUsePreviewValueAsDefault=true, inputName=isVehicle, inputType=FunctionInput_Scalar, sortPriority=5}
  n14 = ScalarParameter() {defaultValue=1, group=Wrap_Squash, parameterName=SQ_SwirlRadiusMech}
  n15 = ScalarParameter() {defaultValue=0.5, group=Wrap_Squash, parameterName=SQ_SwirlRadiusCharacter}
  n16 = NamedRerouteUsage()
  n17 = Constant()
  n18 = FunctionInput(Preview=@n17) {bUsePreviewValueAsDefault=true, inputName=isMech, inputType=FunctionInput_Scalar, sortPriority=4}
  n19 = SphereMask(
   A=@n7
   B=ComponentMask(None=Constant3Vector() {constant=<rgba 0.5 0.5 0 0>}) {g=true, r=true}
   Radius=LinearInterpolate(A=LinearInterpolate(A=@n11, B=@n12, Alpha=@n13), B=LinearInterpolate(A=@n14, B=@n15, Alpha=@n16), Alpha=@n18)
   Hardness=Constant()
  )
  n20 = Multiply(A=@n19, B=ScalarParameter() {defaultValue=15, group=Wrap_Squash, parameterName=SQ_SwirlCoreStrength})
  n21 = Multiply(A=@n10, B=Rotator(Coordinate=@n7, Time=@n20))
  n22 = Multiply(A=@n21, B=Constant() {r=0.75}) {constB=0.75}
  n23 = VectorParameter() {group=Wrap_Squash, parameterName=SQ_Clouds_B_Speed, defaultValue=<rgba -0.08 0.02 0 1>}
  n24 = Multiply(
   A=Time()
   B=ComponentMask(None=VectorParameter() {group=Wrap_Squash, parameterName=SQ_Clouds_WarpB_Speed, defaultValue=<rgba -0.1 -0.03 0 1>}.RGB) {g=true, r=true}
  )
  n25 = TextureObject() {samplerType=SAMPLERTYPE_LinearColor, texture=/Game/Global/Textures/Default/Masks_and_Gradients/T_Wrap_Squash_M}
  n26 = Subtract(A=TextureSample(UVs=Add(A=@n21, B=@n24), Tex=@n25).R) {constB=0.2}
  n27 = ScalarParameter() {defaultValue=0.25, group=Wrap_Squash, parameterName=SQ_Clouds_WarpB_Amount}
  n28 = TextureSample(UVs=Add(A=Panner(Coordinate=@n22, Speed=@n23.RGB), B=Multiply(A=@n26, B=@n27) {constB=0.1}), Tex=@n25) {samplerType=SAMPLERTYPE_LinearColor}
  n29 = Multiply(A=@n10, B=Rotator(Coordinate=@n7, Time=@n20))
  n30 = Panner(Coordinate=@n29, Speed=VectorParameter() {group=Wrap_Squash, parameterName=SQ_Clouds_A_Speed, defaultValue=<rgba 0.1 -0.03 0 1>}.RGB)
  n31 = VectorParameter() {group=Wrap_Squash, parameterName=SQ_Clouds_WarpA_Speed, defaultValue=<rgba -0.1 -0.03 0 1>}
  n32 = Add(A=@n29, B=Multiply(A=Time(), B=ComponentMask(None=@n31.RGB) {g=true, r=true}))
  n33 = Multiply(
   A=Subtract(A=TextureSample(UVs=@n32, Tex=@n25).R) {constB=0.2}
   B=ScalarParameter() {defaultValue=0.25, group=Wrap_Squash, parameterName=SQ_Clouds_WarpA_Amount}
  ) {constB=0.1}
  n34 = ComponentMask(None=TextureSample(UVs=Add(A=@n30, B=@n33), Tex=@n25) {samplerType=SAMPLERTYPE_LinearColor}.A) {r=true}
  n35 = ScalarParameter() {defaultValue=1, group=Wrap_Squash, parameterName=SQ_Clouds_Masking_Multipler}
  n36 = Saturate(None=Add(A=Multiply(A=@n34, B=@n35) {constB=1.5}, B=ScalarParameter() {group=Wrap_Squash, parameterName=SQ_Clouds_Masking_contrast}) {constB=-0.5})
  n37 = Time()
  n38 = Multiply(A=@n37, B=ScalarParameter() {defaultValue=2, group=Wrap_Squash, parameterName=SQ_SwirlCoreA_Speed})
  n39 = Multiply(A=@n19, B=ScalarParameter() {defaultValue=15, group=Wrap_Squash, parameterName=SQ_SwirlCoreInnerStrength})
  n40 = Multiply(A=@n10, B=Rotator(Coordinate=@n7, Time=Add(A=@n38, B=@n39)))
  n41 = Panner(Coordinate=@n40, Speed=VectorParameter() {group=Wrap_Squash, parameterName=SQ_Clouds_A_Speed, defaultValue=<rgba 0.1 -0.03 0 1>}.RGB)
  n42 = VectorParameter() {group=Wrap_Squash, parameterName=SQ_Clouds_WarpA_Speed, defaultValue=<rgba -0.1 -0.03 0 1>}
  n43 = Add(A=@n40, B=Multiply(A=Time(), B=ComponentMask(None=@n42.RGB) {g=true, r=true}))
  n44 = Multiply(
   A=Subtract(A=TextureSample(UVs=@n43, Tex=@n25) {texture=/Game/Weapons/WeaponSkins/Wraps/Textures/Squash/T_Wrap_Squash_DS}.R) {constB=0.2}
   B=ScalarParameter() {defaultValue=0.25, group=Wrap_Squash, parameterName=SQ_Clouds_WarpA_Amount}
  ) {constB=0.1}
  n45 = ComponentMask(None=TextureSample(UVs=Add(A=@n41, B=@n44), Tex=@n25) {samplerType=SAMPLERTYPE_LinearColor}.A) {r=true}
  n46 = Multiply(A=@n37, B=ScalarParameter() {defaultValue=3, group=Wrap_Squash, parameterName=SQ_SwirlCoreB_Speed})
  n47 = Multiply(A=@n10, B=Rotator(Coordinate=@n7, Time=Add(A=@n46, B=@n39)))
  n48 = Multiply(A=@n47, B=Constant() {r=0.75}) {constB=0.75}
  n49 = VectorParameter() {group=Wrap_Squash, parameterName=SQ_Clouds_B_Speed, defaultValue=<rgba -0.08 0.02 0 1>}
  n50 = Multiply(
   A=Time()
   B=ComponentMask(None=VectorParameter() {group=Wrap_Squash, parameterName=SQ_Clouds_WarpB_Speed, defaultValue=<rgba -0.1 -0.03 0 1>}.RGB) {g=true, r=true}
  )
  n51 = Subtract(A=TextureSample(UVs=Add(A=@n47, B=@n50), Tex=@n25) {texture=&T_Wrap_Squash_DS}.R) {constB=0.2}
  n52 = ScalarParameter() {defaultValue=0.25, group=Wrap_Squash, parameterName=SQ_Clouds_WarpB_Amount}
  n53 = TextureSample(UVs=Add(A=Panner(Coordinate=@n48, Speed=@n49.RGB), B=Multiply(A=@n51, B=@n52) {constB=0.1}), Tex=@n25) {samplerType=SAMPLERTYPE_LinearColor}
  n54 = ScalarParameter() {defaultValue=1, group=Wrap_Squash, parameterName=SQ_Clouds_Masking_Multipler}
  n55 = Saturate(None=Add(A=Multiply(A=@n45, B=@n54) {constB=1.5}, B=ScalarParameter() {group=Wrap_Squash, parameterName=SQ_Clouds_Masking_contrast}) {constB=-0.5})
  n56 = Saturate(None=Multiply(A=@n19, B=ScalarParameter() {defaultValue=2, group=Wrap_Squash, parameterName=SQ_SwirlCoreMask_Intensfie}))
  n57 = LinearInterpolate(A=LinearInterpolate(A=@n28.A, B=@n34, Alpha=@n36), B=LinearInterpolate(A=@n45, B=@n53.A, Alpha=@n55), Alpha=@n56)
  n58 = Multiply(A=@n57) {constB=0.05}
  n59 = Add(A=Add(A=Multiply(A=@n5, B=@n6), B=Multiply(A=@n8.Rotated Values, B=@n9)), B=@n58)
  n60 = TextureObject() {texture=/Game/Weapons/WeaponSkins/Wraps/Textures/Squash/T_Wrap_Squash_M_2}
  n61 = Multiply(A=@n8.Rotated Values, B=ScalarParameter() {defaultValue=1.5, group=Wrap_Squash, parameterName=SQ_Rain_Scale_B})
  n62 = Multiply(
   A=Time()
   B=ComponentMask(None=VectorParameter() {group=Wrap_Squash, parameterName=SQ_Rain_Speed_B, defaultValue=<rgba 0.01 -0.5 0 1>}.RGB) {g=true, r=true}
  )
  n63 = TextureSample(UVs=Add(A=Add(A=@n61, B=@n62), B=@n58), Tex=@n60)
  n64 = ScalarParameter() {defaultValue=0.5, group=Wrap_Squash, parameterName=SQ_Rain_B_opacity}
  n65 = Saturate(None=Add(A=TextureSample(UVs=@n59, Tex=@n60).R, B=Multiply(A=@n63.R, B=@n64) {constB=0.25}))
  n66 = VectorParameter() {group=Wrap_Squash, parameterName=SQ_Rain_Color, defaultValue=<rgba 1 1 1 1>}
  n67 = Multiply(
   A=Power(Base=Fresnel() {baseReflectFraction=0, exponent=1}, Exp=ScalarParameter() {defaultValue=2, group=Wrap_Squash, parameterName=SQ_FresnelPower})
   B=VectorParameter() {group=Wrap_Squash, parameterName=SQ_CloudFresnelColor, defaultValue=<rgba 0.2 0.5289 1 1>}.RGB
  )
  n68 = ScalarParameter() {defaultValue=1, group=Wrap_Squash, parameterName=SQ_CloudHeightPowerForDiffuse}
  n69 = LinearInterpolate(
   A=VectorParameter() {group=Wrap_Squash, parameterName=SQ_Clouds_ColorB_Dark, defaultValue=<rgba 0 0 0 1>}.RGB
   B=VectorParameter() {group=Wrap_Squash, parameterName=SQ_Clouds_ColorB_Light, defaultValue=<rgba 0.2961 0.2972 0.3073 1>}.RGB
   Alpha=Power(Base=@n28.A, Exp=@n68)
  )
  n70 = LinearInterpolate(
   A=VectorParameter() {group=Wrap_Squash, parameterName=SQ_Clouds_ColorA_Dark, defaultValue=<rgba 0.026 0.026 0.026 1>}.RGB
   B=VectorParameter() {group=Wrap_Squash, parameterName=SQ_Clouds_ColorA_Light, defaultValue=<rgba 0.6354 0.6354 0.6354 1>}.RGB
   Alpha=Power(Base=@n34, Exp=@n68)
  )
  n71 = ScalarParameter() {defaultValue=1, group=Wrap_Squash, parameterName=SQ_CloudHeightPowerForDiffuse}
  n72 = LinearInterpolate(
   A=VectorParameter() {group=Wrap_Squash, parameterName=SQ_Clouds_ColorB_Dark, defaultValue=<rgba 0 0 0 1>}.RGB
   B=VectorParameter() {group=Wrap_Squash, parameterName=SQ_Clouds_ColorB_Light, defaultValue=<rgba 0.2961 0.2972 0.3073 1>}.RGB
   Alpha=Power(Base=@n53.A, Exp=@n71)
  )
  n73 = LinearInterpolate(
   A=VectorParameter() {group=Wrap_Squash, parameterName=SQ_Clouds_ColorA_Dark, defaultValue=<rgba 0.026 0.026 0.026 1>}.RGB
   B=VectorParameter() {group=Wrap_Squash, parameterName=SQ_Clouds_ColorA_Light, defaultValue=<rgba 0.6354 0.6354 0.6354 1>}.RGB
   Alpha=Power(Base=@n45, Exp=@n71)
  )
  n74 = Multiply(
   A=LinearInterpolate(A=LinearInterpolate(A=@n69, B=@n70, Alpha=@n36), B=LinearInterpolate(A=@n72, B=@n73, Alpha=@n55), Alpha=@n56)
   B=VectorParameter() {group=Wrap_Squash, parameterName=SQ_Clouds_Master_Color_Tent, defaultValue=<rgba 1 1 1 1>}.RGB
  )
  n75 = Multiply(A=Multiply(A=@n67, B=ScalarParameter() {defaultValue=0.3, group=Wrap_Squash, parameterName=SQ_FresnelDiffuseAmount}), B=@n74)
  n76 = Multiply(A=@n67, B=ScalarParameter() {defaultValue=0.05, group=Wrap_Squash, parameterName=SQ_FresnelEmissiveAmount})
  n77 = OneMinus(None=@n57)
  n78 = ScalarParameter() {defaultValue=5.0061, group=Wrap_Squash, parameterName=SQ_LNG-UVOffset}
  n79 = Multiply(A=Time(), B=ScalarParameter() {defaultValue=0.8, group=Wrap_Squash, parameterName=SQ_LNG-TimeShift})
  n80 = TextureObject() {texture=/Game/Weapons/WeaponSkins/Wraps/Textures/Squash/T_Wrap_Squash_Grad}
  n81 = TextureSample(UVs=Multiply(A=ScalarParameter() {defaultValue=0.33, group=Wrap_Squash, parameterName=SQ_LNG-TimeShiftOffset}, B=Ceil(None=@n79)), Tex=@n80) {samplerType=SAMPLERTYPE_Masks}
  n82 = Multiply(A=@n78, B=@n81.G)
  n83 = Multiply(A=@n78, B=@n81.B)
  n84 = ScalarParameter() {defaultValue=1, group=Wrap_Squash, parameterName=SQ_LNG-MaskTile}
  n85 = TextureSample(UVs=Add(A=AppendVector(A=@n82, B=@n83), B=Multiply(A=@n7, B=@n84)), Tex=@n60) {samplerType=SAMPLERTYPE_LinearColor}
  n86 = Frac(None=@n79)
  n87 = OneMinus(None=@n86)
  n88 = Saturate(None=Multiply(A=Multiply(A=@n86, B=@n87), B=ScalarParameter() {defaultValue=0.5, group=Wrap_Squash, parameterName=SQ_LNG-FadeShiftSpeed}))
  n89 = Time()
  n90 = ScalarParameter() {defaultValue=0.25, group=Wrap_Squash, parameterName=SQ_LNG-RandomOffset}
  n91 = TextureSample(UVs=Multiply(A=Add(A=@n89, B=@n90), B=ScalarParameter() {defaultValue=0.2, group=Wrap_Squash, parameterName=SQ_LNG-RandomSpeed}), Tex=@n80) {samplerType=SAMPLERTYPE_Masks}
  n92 = Subtract(A=Time(), B=ScalarParameter() {defaultValue=-9999999, group=00-Reactive-GTFB, parameterName=00-GameTimeFromBP})
  n93 = ScalarParameter() {defaultValue=0.65, group=Wrap_Squash, parameterName=SQ_Lightning_TimeFromBP_Duration}
  n94 = Power(
   Base=Saturate(None=Divide(A=@n92, B=@n93))
   Exp=ScalarParameter() {defaultValue=0.5, group=Wrap_Squash, parameterName=SQ_Lightning_TimeFromBP_Curve}
  )
  n95 = Multiply(A=@n94, B=Multiply(A=Multiply(A=@n85.A, B=@n88), B=Multiply(A=@n91.R, B=@n91.R))) {constA=1}
  n96 = Multiply(
   A=VectorParameter() {group=Wrap_Squash, parameterName=SQ_LNG_Color_A, defaultValue=<rgba 10 9.3384 2.8314 1>}.RGB
   B=Multiply(A=@n95, B=Saturate(None=@n77))
  )
  n97 = LinearInterpolate(
   A=ScalarParameter() {defaultValue=1, group=Wrap_Squash, parameterName=SQ_Lightning_Tile_Weapon}
   B=ScalarParameter() {defaultValue=1, group=Wrap_Squash, parameterName=SQ_Lightning_Tile_Vehicle}
   Alpha=@n13
  )
  n98 = ScalarParameter() {defaultValue=1, group=Wrap_Squash, parameterName=SQ_Lightning_Tile_Mech}
  n99 = TextureSample(
   UVs=Multiply(A=@n7, B=LinearInterpolate(A=@n97, B=@n98, Alpha=@n18))
   Tex=TextureObject() {samplerType=SAMPLERTYPE_LinearColor, texture=/Game/Weapons/WeaponSkins/Wraps/Textures/Squash/T_Wrap_Squash_M_1}
  )
  n100 = Time()
  n101 = Constant() {r=0.8}
  n102 = Ceil(None=Sine(None=Multiply(A=@n100, B=@n101)))
  n103 = Sine(None=Multiply(A=@n100, B=Constant() {r=0.6}))
  n104 = LinearInterpolate(
   A=LinearInterpolate(A=ComponentMask(None=@n99.RGB) {r=true}, B=ComponentMask(None=@n99.RGB) {g=true}, Alpha=Saturate(None=@n102))
   B=ComponentMask(None=@n99.RGB) {b=true}
   Alpha=Saturate(None=Ceil(None=@n103))
  )
  n105 = LinearInterpolate(
   A=ScalarParameter() {defaultValue=5, group=Wrap_Squash, parameterName=SQ_Lightning_height_Max}
   B=ScalarParameter() {defaultValue=-1, group=Wrap_Squash, parameterName=SQ_Lightning_height_Min}
   Alpha=@n57
  ) {constA=5, constB=0}
  n106 = Multiply(
   A=Multiply(A=Saturate(None=@n104), B=Saturate(None=@n105))
   B=VectorParameter() {group=Wrap_Squash, parameterName=SQ_Lightning_Color, defaultValue=<rgba 1.8672 2.5385 3 1>}.RGB
  )
  n107 = Saturate(None=Divide(A=@n92) {constB=0.15})
  n108 = Multiply(
   A=Add(A=Time(), B=ScalarParameter() {defaultValue=0.35, group=Wrap_Squash, parameterName=SQ_Lightning_RandomOffset})
   B=ScalarParameter() {defaultValue=0.2, group=Wrap_Squash, parameterName=SQ_Lightning_RandomSpeed}
  )
  n109 = Multiply(A=OneMinus(None=@n94), B=Add(A=OneMinus(None=@n107), B=TextureSample(UVs=@n108, Tex=@n80) {samplerType=SAMPLERTYPE_Masks}.R))
  n110 = SetMaterialAttributes!(
   MaterialAttributes=@n4.MaterialAttributes
   BaseColor=Add(A=Multiply(A=@n65, B=@n66.RGB), B=Add(A=@n75, B=@n74))
   Specular=ScalarParameter() {defaultValue=0.5, group=Wrap_Squash, parameterName=SQ_Spec, sortPriority=-10}
   Roughness=ScalarParameter() {defaultValue=0.6, group=Wrap_Squash, parameterName=SQ_Roughness, sortPriority=-10}
   Normal
   Metallic=Constant()
   EmissiveColor=Add(A=Multiply(A=@n106, B=@n109), B=Add(A=@n96, B=@n76))
  ) {attributeSetTypes [
    #2
    #4
    #5
    #3
    #6
    #1
   ]}
  n111 = Constant() {r=1}
  n112 = Saturate(None=Add(
    A=FunctionInput(Preview=@n111) {bUsePreviewValueAsDefault=true, inputName="Customization Mask (_CM B)", inputType=FunctionInput_Scalar, sortPriority=24}
    B=ScalarParameter() {group=Wrap_Squash, parameterName=SQ_Remove_Customization Mask, sortPriority=-10}
   ))
  FunctionOutput(None=StaticSwitchParameter(True=BlendMaterialAttributes(A=@n4.MaterialAttributes, B=@n110, Alpha=@n112), False=@n3) {group=Customization_Wrap, parameterName=Wrap_Squash})
  n113 = PreSkinnedLocalBounds()
  n114 = Add(A=@n113.Min, B=FunctionInput() {bUsePreviewValueAsDefault=true, inputName=Object Bounds Min Adjustment, sortPriority=50}) {constB=0}
  n115 = FunctionInput() {bUsePreviewValueAsDefault=true, inputName=Object Bounds Max Adjustment, sortPriority=50}
  n116 = Multiply(A=Subtract(A=Add(A=@n113.Max, B=@n115) {constB=0}, B=@n114)) {constB=0.5}
  n117 = LinearInterpolate(A=TransformPosition(None=Add(A=@n114, B=@n116)) {transformType=TRANSFORMPOSSOURCE_World}, B=ObjectPositionWS(), Alpha=@n18)
  n118 = Normalize(VectorInput=Subtract(A=@n117, B=CameraPositionWS()))
  n119 = CrossProduct(A=@n118, B=Constant3Vector() {constant=<rgba 0 0 1 1>})
  n120 = MaterialFunctionCall() {materialFunction=/Engine/Functions/Engine_MaterialFunctions02/WorldPositionOffset/ObjectScale}
  n121 = MaterialFunctionCall(Float3=@n113.Extents) {materialFunction=&BreakOutFloat3Components}
  n122 = Max(A=Max(A=@n121.R, B=@n121.R), B=@n121.B)
  LinearInterpolate(
   A=MaterialFunctionCall!(Plane Normal Axis=@n118, Plane X Axis=Normalize(VectorInput=@n119), Plane Center=@n117, UV Scale=Multiply(A=@n120.Scale X, B=@n122)) {materialFunction=/Game/Athena/Environments/Nexus/Materials/VirtualPlaneCoordinatesAbs}.UVs
   B=@n7
   Alpha=NamedRerouteUsage()
  )
  n123 = TransformPosition(None=Constant3Vector()) {transformType=TRANSFORMPOSSOURCE_World}
  n124 = Normalize(VectorInput=Subtract(A=@n123, B=CameraPositionWS()))
  n125 = Constant3Vector() {constant=<rgba 0 0 1 1>}
  n126 = FunctionInput() {bUsePreviewValueAsDefault=true, inputName=Object-SafeZone-Virtual-Pivot, inputType=FunctionInput_Vector4, sortPriority=105}
  n127 = MaterialFunctionCall() {materialFunction=&ObjectScale}
  n128 = MaterialFunctionCall(Float3=PreSkinnedLocalBounds().Extents) {materialFunction=&BreakOutFloat3Components}
  n129 = Max(A=Max(A=@n128.R, B=@n128.R), B=@n128.B)
  MaterialFunctionCall(
   Plane Normal Axis=@n124
   Plane X Axis=Normalize(VectorInput=CrossProduct(A=@n124, B=@n125))
   Plane Center=Add(A=@n123, B=ComponentMask(None=@n126) {b=true, g=true, r=true})
   UV Scale=Multiply(A=ComponentMask(None=@n126) {a=true}, B=Multiply(A=@n127.Scale X, B=@n129))
  ) {materialFunction=&VirtualPlaneCoordinatesAbs}
  n130 = Multiply(A=OneMinus(None=@n95)) {constB=10.75}
  Multiply(A=Saturate(None=Add(A=@n130) {constB=-10}), B=@n74)
  NamedRerouteDeclaration(None=FunctionInput(Preview=@n17) {bUsePreviewValueAsDefault=true, inputName=IsCharacter, inputType=FunctionInput_Scalar, sortPriority=4}) {name=IsCharacter, nodeColor=<rgba 1 0.3047 0 1>}
  n131 = FunctionInput() {bUsePreviewValueAsDefault=true, inputName=Object-SafeZone-Virtual-CH-Switch, inputType=FunctionInput_Scalar, sortPriority=105}
  NamedRerouteDeclaration(None=MaterialFunctionCall!(
    Camera _pos
    Axis
    Scale
    Scale_Multplyer
    Object-SafeZone-Virtual-Pivot=@n126
    Object-SafeZone-Virtual-CH-Switch=@n131
    SinglePlane Normal Axis
    SinglePlane X Axis
    PlaneCenter_distortion
   ) {materialFunction=&/MF_SafeZone_VirtualPlane}.Uvs) {name=SafeZoneVP_UVs, nodeColor=<rgba 1 0 0.6562 1>}
 ]}
