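# Star Aces wrap material function for weapons, vehicles and mechs, layering four animated playing cards and two sparkling, scrolling star layers over a gradient.
# Bounds/normal side projection supplies UVs, with vehicle-type inputs selecting tuning.
# Game-time parameters trigger card slides with sine bobbing and a gradient color fade.
# Customization masks blend wrap and secondary material behind a static switch; a second card set and randomized-star branches are unconnected.
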MaterialFunction /Game/Weapons/WeaponSkins/Wraps/MaterialFunctions/MF_Wrap_StarAces
graph {nodes [
  n1 = VectorParameter() {group=Wrap_StarAces, parameterName=SAces_Card_TopRT_Offset_End, defaultValue=<rgba 0 0 0 1>}
  n2 = Time()
  n3 = VectorParameter() {group=Wrap_StarAces, parameterName=SAces_Card__Sine_Speed_Black, defaultValue=<rgba 0 -0.1 0 1>}
  n4 = ScalarParameter() {defaultValue=0.25, group=Wrap_StarAces, parameterName=SAces_Card__Sine_Range_Vehicle}
  n5 = ScalarParameter() {defaultValue=0.25, group=Wrap_StarAces, parameterName=SAces_Card__Sine_Range_VehicleSide}
  n6 = Constant()
  n7 = MaterialFunctionCall(
   R=FunctionInput(Preview=@n6) {inputName=isVehicle, inputType=FunctionInput_Scalar, sortPriority=5}
   G=FunctionInput(Preview=@n6) {bUsePreviewValueAsDefault=true, inputName=Side Project Vehicle, inputType=FunctionInput_Scalar}
   B=FunctionInput(Preview=@n6) {bUsePreviewValueAsDefault=true, inputName=isMech, inputType=FunctionInput_Scalar, sortPriority=4}
  ) {materialFunction=&AppendMany}
  n8 = MaterialFunctionCall(Color=@n7.RGB) {materialFunction=&SplitComponents}
  n9 = LinearInterpolate(
   A=LinearInterpolate(
    A=ScalarParameter() {defaultValue=0.25, group=Wrap_StarAces, parameterName=SAces_Card__Sine_Range_Weapon}
    B=LinearInterpolate(A=@n4, B=@n5, Alpha=@n8.G)
    Alpha=@n8.R
   )
   B=ScalarParameter() {defaultValue=0.25, group=Wrap_StarAces, parameterName=SAces_Card__Sine_Range_Mech}
   Alpha=@n8.B
  )
  n10 = Multiply(A=Sine(None=Multiply(A=@n2, B=@n3.RGB)), B=@n9)
  n11 = Divide(
   A=Subtract(A=Time(), B=ScalarParameter() {defaultValue=-9999999, group=Wrap_StarAces, parameterName=SAces_Game_TimeBPG, sortPriority=100})
   B=ScalarParameter() {defaultValue=0.25, group=Wrap_StarAces, parameterName=SAces_Duration, sortPriority=100}
  )
  n12 = ComponentMask(None=LinearInterpolate(
    A=VectorParameter() {group=Wrap_StarAces, parameterName=SAces_Card_TopRT_Offset, defaultValue=<rgba 0 0 0 1>}.RGB
    B=Add(A=@n10, B=@n1.RGB)
    Alpha=Saturate(None=@n11)
   )) {g=true, r=true}
  n13 = ScalarParameter() {defaultValue=1, group=Wrap_StarAces, parameterName=SAces_UV_Scale_Weapon}
  n14 = MaterialFunctionCall(Color=@n7.RGB) {materialFunction=&SplitComponents}
  n15 = LinearInterpolate(
   A=ScalarParameter() {defaultValue=1, group=Wrap_StarAces, parameterName=SAces_UV_Scale_Vehicle}
   B=ScalarParameter() {defaultValue=1, group=Wrap_StarAces, parameterName=SAces_UV_Scale_VehicleSide}
   Alpha=@n14.G
  )
  n16 = VectorParameter() {group=Wrap_StarAces, parameterName=SAces_UV_Offset_Vehicle, defaultValue=<rgba 0 0 0 1>}
  n17 = VectorParameter() {group=Wrap_StarAces, parameterName=SAces_UV_Offset_VehicleSide, defaultValue=<rgba 0 0 0 1>}
  n18 = MaterialFunctionCall(Color=@n7.RGB) {materialFunction=&SplitComponents}
  n19 = LinearInterpolate(
   A=LinearInterpolate(
    A=VectorParameter() {group=Wrap_StarAces, parameterName=SAces_UV_Offset_Weapon, defaultValue=<rgba 0 0 0 1>}.RGB
    B=LinearInterpolate(A=@n16.RGB, B=@n17.RGB, Alpha=@n18.G)
    Alpha=@n18.R
   )
   B=VectorParameter() {group=Wrap_StarAces, parameterName=SAces_UV_Offset_Mech, defaultValue=<rgba 0 0 0 1>}.RGB
   Alpha=@n18.B
  )
  n20 = Constant3Vector() {constant=<rgba 0 1 0 1>}
  n21 = Constant3Vector() {constant=<rgba 0 0 0 1>}
  n22 = MaterialFunctionCall(Color=@n7.RGB) {materialFunction=&SplitComponents}
  n23 = LinearInterpolate(A=LinearInterpolate(A=@n20, B=LinearInterpolate(A=@n21, B=@n20, Alpha=@n22.G), Alpha=@n22.R), B=@n20, Alpha=@n22.B)
  n24 = VertexInterpolator(VS=PreSkinnedPosition())
  n25 = PreSkinnedLocalBounds()
  n26 = FunctionInput(Preview=Divide(A=Subtract(A=@n24.PS, B=@n25.Min), B=Subtract(A=@n25.Max, B=@n25.Max))) {bUsePreviewValueAsDefault=true, inputName=Calcated_Bounds, sortPriority=11}
  n27 = FunctionInput(Preview=VertexInterpolator(VS=PreSkinnedNormal()).PS) {bUsePreviewValueAsDefault=true, inputName=Pre-Skinned Local Normal, sortPriority=10}
  n28 = MaterialFunctionCall(Color=@n7.RGB) {materialFunction=&SplitComponents}
  n29 = ScalarParameter() {group=Wrap_StarAces, parameterName=SAces_UV_Rotation_Weapon}
  n30 = MaterialFunctionCall(Color=@n7.RGB) {materialFunction=&SplitComponents}
  n31 = LinearInterpolate(
   A=ScalarParameter() {group=Wrap_StarAces, parameterName=SAces_UV_Rotation_Vehicle}
   B=ScalarParameter() {group=Wrap_StarAces, parameterName=SAces_UV_Rotation_VehicleSide}
   Alpha=@n30.G
  )
  n32 = MaterialFunctionCall(
   UVs=Add(
    A=ComponentMask(None=@n23) {g=true, r=true}
    B=MaterialFunctionCall!(
     For_Safe_Zones
     Normal_Map
     Caculated_Bounds=@n26
     Pre_Skinned_Normals=@n27
     Is Vehicle=@n28.R
     Is VehicleSide=@n28.R
     Is Mech=@n28.R
     Vehicle_Rotation
     UV_Rotation Bypass
     Mech_Rotation
     Uvs_Offset
     Offset-after rotation
     Uvs_TileRate
    ) {materialFunction=&/MF_Projection_Side_Normal}.UVs No Offsets/Tiling
   )
   "Rotation Angle (0-1)"=LinearInterpolate(
    A=LinearInterpolate(A=@n29, B=@n31, Alpha=@n30.R)
    B=ScalarParameter() {defaultValue=0.25, group=Wrap_StarAces, parameterName=SAces_UV_Rotation_Mech}
    Alpha=@n30.B
   )
  ) {materialFunction=&CustomRotator}
  n33 = Multiply(
   A=LinearInterpolate(
    A=LinearInterpolate(A=@n13, B=@n15, Alpha=@n14.R)
    B=ScalarParameter() {defaultValue=1, group=Wrap_StarAces, parameterName=SAces_UV_Scale_Mech}
    Alpha=@n14.B
   )
   B=Add(A=ComponentMask(None=@n19) {g=true, r=true}, B=@n32.Rotated Values)
  )
  n34 = Multiply(A=Multiply(A=Add(A=@n33, B=@n12)) {constB=0.25}) {constB=4}
  n35 = Constant2Vector() {r=0.5}
  n36 = ScalarParameter() {defaultValue=0.5, group=Wrap_StarAces, parameterName=SAces_Card__Scale_Vehicle}
  n37 = ScalarParameter() {defaultValue=0.5, group=Wrap_StarAces, parameterName=SAces_Card__Scale_VehicleSide}
  n38 = MaterialFunctionCall(Color=@n7.RGB) {materialFunction=&SplitComponents}
  n39 = LinearInterpolate(
   A=LinearInterpolate(
    A=ScalarParameter() {defaultValue=0.5, group=Wrap_StarAces, parameterName=SAces_Card__Scale_Weapon}
    B=LinearInterpolate(A=@n36, B=@n37, Alpha=@n38.G)
    Alpha=@n38.R
   )
   B=ScalarParameter() {defaultValue=0.5, group=Wrap_StarAces, parameterName=SAces_Card__Scale_Mech}
   Alpha=@n38.B
  )
  n40 = Divide(A=Add(A=Multiply(A=@n34) {constB=0.5}, B=@n35), B=@n39)
  n41 = Constant2Vector() {g=0.25, r=0.75}
  n42 = Constant3Vector() {constant=<rgba 0.25 0.25 0 1>}
  n43 = LinearInterpolate(A=Time(), Alpha=ScalarParameter() {group=Wrap_StarAces, parameterName="SAces__Card_Speed_Disable 1=Active"})
  n44 = MaterialFunctionCall(
   UVs=Subtract(A=Add(A=@n40, B=@n41), B=Divide(A=@n41, B=@n39) {desc=Scale the center offset by the scaling factor}) {desc=Re offset to maintain center}
   Rotation Center=Add(A=@n35, B=ComponentMask(None=@n42) {g=true, r=true})
   "Rotation Angle (0-1)"=Multiply(A=@n43, B=ScalarParameter() {group=Wrap_StarAces, parameterName=SAces_Card_TopRT_Rotation})
  ) {materialFunction=&CustomRotator}
  n45 = TextureObjectParameter() {group=Wrap_StarAces, parameterName="SAces_Lerp__Text_Mask A=Card Mask ", texture=&T_BlacksRGB}
  n46 = TextureSample(UVs=Add(A=Constant2Vector() {g=0.25, r=-0.25}, B=@n44.Rotated Values), Tex=@n45)
  n47 = VectorParameter() {group=Wrap_StarAces, parameterName=SAces_Card_TopLft_Offset_End, defaultValue=<rgba 0 0 0 1>}
  n48 = VectorParameter() {group=Wrap_StarAces, parameterName=SAces_Card__Sine_Speed_Red, defaultValue=<rgba 0 0.1 0 1>}
  n49 = Multiply(A=Sine(None=Multiply(A=@n2, B=@n48.RGB)), B=@n9)
  n50 = Divide(
   A=Subtract(A=Time(), B=ScalarParameter() {defaultValue=-9999999, group=Wrap_StarAces, parameterName=SAces_Game_TimeBPF, sortPriority=100})
   B=ScalarParameter() {defaultValue=0.25, group=Wrap_StarAces, parameterName=SAces_Duration, sortPriority=100}
  )
  n51 = ComponentMask(None=LinearInterpolate(
    A=VectorParameter() {group=Wrap_StarAces, parameterName=SAces_Card_TopLft_Offset, defaultValue=<rgba 0 0 0 1>}.RGB
    B=Add(A=@n47.RGB, B=@n49)
    Alpha=Saturate(None=@n50)
   )) {g=true, r=true}
  n52 = Multiply(A=Multiply(A=Add(A=@n33, B=@n51)) {constB=0.25}) {constB=4}
  n53 = Constant2Vector()
  n54 = Divide(A=Add(A=Multiply(A=@n52) {constB=0.5}, B=@n53), B=@n39)
  n55 = Constant() {r=0.25}
  n56 = Constant3Vector() {constant=<rgba 0.25 0.25 0 1>}
  n57 = MaterialFunctionCall(
   UVs=Subtract(A=Add(A=@n54, B=@n55), B=Divide(A=@n55, B=@n39) {desc=Scale the center offset by the scaling factor}) {desc=Re offset to maintain center}
   Rotation Center=Add(A=@n53, B=ComponentMask(None=@n56) {g=true, r=true})
   "Rotation Angle (0-1)"=Multiply(A=@n43, B=ScalarParameter() {group=Wrap_StarAces, parameterName=SAces_Card_TopLft_Rotation})
  ) {materialFunction=&CustomRotator}
  n58 = TextureSample(UVs=Add(A=Constant2Vector() {g=0.25, r=0.25}, B=@n57.Rotated Values), Tex=@n45)
  n59 = Add(A=@n58.A, B=@n46.A)
  n60 = VectorParameter() {group=Wrap_StarAces, parameterName=SAces_Card_BotLft_Offset_End, defaultValue=<rgba 0 0 0 1>}
  n61 = Divide(
   A=Subtract(A=Time(), B=ScalarParameter() {defaultValue=-9999999, group=Wrap_StarAces, parameterName=SAces_Game_TimeBPH, sortPriority=100})
   B=ScalarParameter() {defaultValue=0.25, group=Wrap_StarAces, parameterName=SAces_Duration, sortPriority=100}
  )
  n62 = ComponentMask(None=LinearInterpolate(
    A=VectorParameter() {group=Wrap_StarAces, parameterName=SAces_Card_BotLft_Offset, defaultValue=<rgba 0 0 0 1>}.RGB
    B=Add(A=@n49, B=@n60.RGB)
    Alpha=Saturate(None=@n61)
   )) {g=true, r=true}
  n63 = Multiply(A=Multiply(A=Add(A=@n33, B=@n62)) {constB=0.25}) {constB=4}
  n64 = Constant2Vector() {g=0.5}
  n65 = Divide(A=Add(A=Multiply(A=@n63) {constB=0.5}, B=@n64), B=@n39)
  n66 = Constant2Vector() {g=0.75, r=0.25}
  n67 = Constant3Vector() {constant=<rgba 0.25 0.25 0 1>}
  n68 = MaterialFunctionCall(
   UVs=Subtract(A=Add(A=@n65, B=@n66), B=Divide(A=@n66, B=@n39) {desc=Scale the center offset by the scaling factor}) {desc=Re offset to maintain center}
   Rotation Center=Add(A=@n64, B=ComponentMask(None=@n67) {g=true, r=true})
   "Rotation Angle (0-1)"=Multiply(A=@n43, B=ScalarParameter() {group=Wrap_StarAces, parameterName=SAces_Card_BotLft_Rotation})
  ) {materialFunction=&CustomRotator}
  n69 = TextureSample(UVs=Add(A=Constant2Vector() {g=-0.25, r=0.25}, B=@n68.Rotated Values), Tex=@n45)
  n70 = VectorParameter() {group=Wrap_StarAces, parameterName=SAces_Card_BotRT_Offset_End, defaultValue=<rgba 0 0 0 1>}
  n71 = Divide(
   A=Subtract(A=Time(), B=ScalarParameter() {defaultValue=-9999999, group=Wrap_StarAces, parameterName=SAces_Game_TimeBPI, sortPriority=100})
   B=ScalarParameter() {defaultValue=0.25, group=Wrap_StarAces, parameterName=SAces_Duration, sortPriority=100}
  )
  n72 = ComponentMask(None=LinearInterpolate(
    A=VectorParameter() {group=Wrap_StarAces, parameterName=SAces_Card_BotRT_Offset, defaultValue=<rgba 0 0 0 1>}.RGB
    B=Add(A=@n10, B=@n70.RGB)
    Alpha=Saturate(None=@n71)
   )) {g=true, r=true}
  n73 = Multiply(A=Multiply(A=Add(A=@n33, B=@n72)) {constB=0.25}) {constB=4}
  n74 = Constant2Vector() {g=0.5, r=0.5}
  n75 = Divide(A=Add(A=Multiply(A=@n73) {constB=0.5}, B=@n74), B=@n39)
  n76 = Constant2Vector() {g=0.75, r=0.75}
  n77 = Constant3Vector() {constant=<rgba 0.25 0.25 0 1>}
  n78 = MaterialFunctionCall(
   UVs=Subtract(A=Add(A=@n75, B=@n76), B=Divide(A=@n76, B=@n39) {desc=Scale the center offset by the scaling factor}) {desc=Re offset to maintain center}
   Rotation Center=Add(A=@n74, B=ComponentMask(None=@n77) {g=true, r=true})
   "Rotation Angle (0-1)"=Multiply(A=@n43, B=ScalarParameter() {group=Wrap_StarAces, parameterName=SAces_Card_BotRT_Rotation})
  ) {materialFunction=&CustomRotator}
  n79 = TextureSample(UVs=Add(A=Constant2Vector() {g=-0.25, r=-0.25}, B=@n78.Rotated Values), Tex=@n45)
  n80 = Add(A=@n69.A, B=@n79.A)
  n81 = VectorParameter() {group=Wrap_StarAces, parameterName=SAces_Card_TopRT_Offset_End, defaultValue=<rgba 0 0 0 1>}
  n82 = Divide(
   A=Subtract(A=Time(), B=ScalarParameter() {defaultValue=-9999999, group=Wrap_StarAces, parameterName=SAces_Game_TimeBPB, sortPriority=100})
   B=ScalarParameter() {defaultValue=0.25, group=Wrap_StarAces, parameterName=SAces_Duration, sortPriority=100}
  )
  n83 = ComponentMask(None=LinearInterpolate(
    A=VectorParameter() {group=Wrap_StarAces, parameterName=SAces_Card_TopRT_Offset, defaultValue=<rgba 0 0 0 1>}.RGB
    B=Add(A=@n10, B=@n81.RGB)
    Alpha=Saturate(None=@n82)
   )) {g=true, r=true}
  n84 = Multiply(A=Multiply(A=Add(A=@n33, B=@n83)) {constB=0.25}) {constB=4}
  n85 = Constant2Vector() {r=0.5}
  n86 = Divide(A=Add(A=Multiply(A=@n84) {constB=0.5}, B=@n85), B=@n39)
  n87 = Constant2Vector() {g=0.25, r=0.75}
  n88 = Constant3Vector() {constant=<rgba 0.25 0.25 0 1>}
  n89 = LinearInterpolate(A=Time(), Alpha=ScalarParameter() {group=Wrap_StarAces, parameterName="SAces__Card_Speed_Disable 1=Active"})
  n90 = MaterialFunctionCall(
   UVs=Subtract(A=Add(A=@n86, B=@n87), B=Divide(A=@n87, B=@n39) {desc=Scale the center offset by the scaling factor}) {desc=Re offset to maintain center}
   Rotation Center=Add(A=@n85, B=ComponentMask(None=@n88) {g=true, r=true})
   "Rotation Angle (0-1)"=Multiply(A=@n89, B=ScalarParameter() {group=Wrap_StarAces, parameterName=SAces_Card_TopRT_Rotation})
  ) {materialFunction=&CustomRotator}
  n91 = TextureSample(UVs=Add(A=Constant2Vector() {g=0.25, r=-0.25}, B=@n90.Rotated Values), Tex=@n45)
  n92 = VectorParameter() {group=Wrap_StarAces, parameterName=SAces_Card_TopLft_Offset_End, defaultValue=<rgba 0 0 0 1>}
  n93 = Divide(
   A=Subtract(A=Time(), B=ScalarParameter() {defaultValue=-9999999, group=Wrap_StarAces, parameterName=SAces_Game_TimeBPA, sortPriority=100})
   B=ScalarParameter() {defaultValue=0.25, group=Wrap_StarAces, parameterName=SAces_Duration, sortPriority=100}
  )
  n94 = ComponentMask(None=LinearInterpolate(
    A=VectorParameter() {group=Wrap_StarAces, parameterName=SAces_Card_TopLft_Offset, defaultValue=<rgba 0 0 0 1>}.RGB
    B=Add(A=@n92.RGB, B=@n49)
    Alpha=Saturate(None=@n93)
   )) {g=true, r=true}
  n95 = Multiply(A=Multiply(A=Add(A=@n33, B=@n94)) {constB=0.25}) {constB=4}
  n96 = Constant2Vector()
  n97 = Divide(A=Add(A=Multiply(A=@n95) {constB=0.5}, B=@n96), B=@n39)
  n98 = Constant() {r=0.25}
  n99 = Constant3Vector() {constant=<rgba 0.25 0.25 0 1>}
  n100 = MaterialFunctionCall(
   UVs=Subtract(A=Add(A=@n97, B=@n98), B=Divide(A=@n98, B=@n39) {desc=Scale the center offset by the scaling factor}) {desc=Re offset to maintain center}
   Rotation Center=Add(A=@n96, B=ComponentMask(None=@n99) {g=true, r=true})
   "Rotation Angle (0-1)"=Multiply(A=@n89, B=ScalarParameter() {group=Wrap_StarAces, parameterName=SAces_Card_TopLft_Rotation})
  ) {materialFunction=&CustomRotator}
  n101 = TextureSample(UVs=Add(A=Constant2Vector() {g=0.25, r=0.25}, B=@n100.Rotated Values), Tex=@n45)
  n102 = VectorParameter() {group=Wrap_StarAces, parameterName=SAces_Card_BotLft_Offset_End, defaultValue=<rgba 0 0 0 1>}
  n103 = Divide(
   A=Subtract(A=Time(), B=ScalarParameter() {defaultValue=-9999999, group=Wrap_StarAces, parameterName=SAces_Game_TimeBPC, sortPriority=100})
   B=ScalarParameter() {defaultValue=0.25, group=Wrap_StarAces, parameterName=SAces_Duration, sortPriority=100}
  )
  n104 = ComponentMask(None=LinearInterpolate(
    A=VectorParameter() {group=Wrap_StarAces, parameterName=SAces_Card_BotLft_Offset, defaultValue=<rgba 0 0 0 1>}.RGB
    B=Add(A=@n49, B=@n102.RGB)
    Alpha=Saturate(None=@n103)
   )) {g=true, r=true}
  n105 = Multiply(A=Multiply(A=Add(A=@n33, B=@n104)) {constB=0.25}) {constB=4}
  n106 = Constant2Vector() {g=0.5}
  n107 = Divide(A=Add(A=Multiply(A=@n105) {constB=0.5}, B=@n106), B=@n39)
  n108 = Constant2Vector() {g=0.75, r=0.25}
  n109 = Constant3Vector() {constant=<rgba 0.25 0.25 0 1>}
  n110 = MaterialFunctionCall(
   UVs=Subtract(A=Add(A=@n107, B=@n108), B=Divide(A=@n108, B=@n39) {desc=Scale the center offset by the scaling factor}) {desc=Re offset to maintain center}
   Rotation Center=Add(A=@n106, B=ComponentMask(None=@n109) {g=true, r=true})
   "Rotation Angle (0-1)"=Multiply(A=@n89, B=ScalarParameter() {group=Wrap_StarAces, parameterName=SAces_Card_BotLft_Rotation})
  ) {materialFunction=&CustomRotator}
  n111 = TextureSample(UVs=Add(A=Constant2Vector() {g=-0.25, r=0.25}, B=@n110.Rotated Values), Tex=@n45)
  n112 = VectorParameter() {group=Wrap_StarAces, parameterName=SAces_Card_BotRT_Offset_End, defaultValue=<rgba 0 0 0 1>}
  n113 = Divide(
   A=Subtract(A=Time(), B=ScalarParameter() {defaultValue=-9999999, group=Wrap_StarAces, parameterName=SAces_Game_TimeBPD, sortPriority=100})
   B=ScalarParameter() {defaultValue=0.25, group=Wrap_StarAces, parameterName=SAces_Duration, sortPriority=100}
  )
  n114 = ComponentMask(None=LinearInterpolate(
    A=VectorParameter() {group=Wrap_StarAces, parameterName=SAces_Card_BotRT_Offset, defaultValue=<rgba 0 0 0 1>}.RGB
    B=Add(A=@n10, B=@n112.RGB)
    Alpha=Saturate(None=@n113)
   )) {g=true, r=true}
  n115 = Multiply(A=Multiply(A=Add(A=@n33, B=@n114)) {constB=0.25}) {constB=4}
  n116 = Constant2Vector() {g=0.5, r=0.5}
  n117 = Divide(A=Add(A=Multiply(A=@n115) {constB=0.5}, B=@n116), B=@n39)
  n118 = Constant2Vector() {g=0.75, r=0.75}
  n119 = Constant3Vector() {constant=<rgba 0.25 0.25 0 1>}
  n120 = MaterialFunctionCall(
   UVs=Subtract(A=Add(A=@n117, B=@n118), B=Divide(A=@n118, B=@n39) {desc=Scale the center offset by the scaling factor}) {desc=Re offset to maintain center}
   Rotation Center=Add(A=@n116, B=ComponentMask(None=@n119) {g=true, r=true})
   "Rotation Angle (0-1)"=Multiply(A=@n89, B=ScalarParameter() {group=Wrap_StarAces, parameterName=SAces_Card_BotRT_Rotation})
  ) {materialFunction=&CustomRotator}
  n121 = TextureSample(UVs=Add(A=Constant2Vector() {g=-0.25, r=-0.25}, B=@n120.Rotated Values), Tex=@n45)
  n122 = Add(A=Add(A=@n101.A, B=@n91.A), B=Add(A=@n111.A, B=@n121.A))
  Add(A=Saturate(None=Add(A=@n59, B=@n80)), B=@n122)
  Constant() {r=1}
  Constant()
  Frac(None=@n84)
  Frac(None=@n95)
  Frac(None=@n34)
  Frac(None=@n73)
  Frac(None=@n63)
  Frac(None=@n115)
  Frac(None=@n105)
  Frac(None=@n52)
  FunctionInput() {bUsePreviewValueAsDefault=true, inputName=Curavture, inputType=FunctionInput_Scalar, sortPriority=24}
  n123 = FunctionInput() {bUsePreviewValueAsDefault=true, inputName=Material, inputType=FunctionInput_MaterialAttributes, sortPriority=-20}
  n124 = VectorParameter() {group=Wrap_StarAces, parameterName=SAces_Grad_Color_alt, defaultValue=<rgba 0.599 0.4283 0.1061 1>}
  n125 = VectorParameter() {group=Wrap_StarAces, parameterName=SAces_Grad_Color, defaultValue=<rgba 0.599 0.4283 0.1061 1>}
  n126 = Subtract(A=Time(), B=ScalarParameter() {defaultValue=-9999999, group=Wrap_StarAces, parameterName=SAces_Game_TimeBPE, sortPriority=100})
  n127 = ScalarParameter() {defaultValue=0.25, group=Wrap_StarAces, parameterName=SAces_Duration_E, sortPriority=100}
  n128 = Saturate(None=Saturate(None=Divide(A=@n126, B=@n127)))
  n129 = VectorParameter() {group=Wrap_StarAces, parameterName="SAces_Foward_Axis_overide  ", sortPriority=0, defaultValue=<rgba 0 0 0 1>}
  n130 = Constant3Vector() {constant=<rgba 0 1 0 0>}
  n131 = ScalarParameter() {group=Wrap_StarAces, parameterName=SAces_Grad_Offset_Weapon}
  n132 = MaterialFunctionCall(Color=@n7.RGB) {materialFunction=&SplitComponents}
  n133 = LinearInterpolate(
   A=ScalarParameter() {group=Wrap_StarAces, parameterName=SAces_Grad_Offset_Vehicle}
   B=ScalarParameter() {group=Wrap_StarAces, parameterName=SAces_Grad_Offset_VehicleSide}
   Alpha=@n132.G
  )
  n134 = ScalarParameter() {defaultValue=0.75, group=Wrap_StarAces, parameterName=SAces_Grad_Falloff_Weapon}
  n135 = MaterialFunctionCall(Color=@n7.RGB) {materialFunction=&SplitComponents}
  n136 = LinearInterpolate(
   A=ScalarParameter() {defaultValue=0.75, group=Wrap_StarAces, parameterName=SAces_Grad_Falloff_Vehicle}
   B=ScalarParameter() {defaultValue=0.75, group=Wrap_StarAces, parameterName=SAces_Grad_Falloff_VehicleSide}
   Alpha=@n135.G
  )
  n137 = MaterialFunctionCall!(
   Bounds=@n26
   ForwardAxis=LinearInterpolate(
    A=FunctionInput(Preview=@n130) {bUsePreviewValueAsDefault=true, inputName=ForwardAxis}
    B=@n129.RGB
    Alpha=ScalarParameter() {group=Wrap_StarAces, parameterName="SAces_Foward_Axis_overide  0=off", sortPriority=0}
   )
   Offset=LinearInterpolate(
    A=LinearInterpolate(A=@n131, B=@n133, Alpha=@n132.R)
    B=ScalarParameter() {group=Wrap_StarAces, parameterName=SAces_Grad_Offset_Mech}
    Alpha=@n132.B
   )
   FallOff=LinearInterpolate(
    A=LinearInterpolate(A=@n134, B=@n136, Alpha=@n135.R)
    B=ScalarParameter() {defaultValue=0.75, group=Wrap_StarAces, parameterName=SAces_Grad_Falloff_Mech}
    Alpha=@n135.B
   )
  ) {materialFunction=&/MF_Wrap_Gradent_Simple}
  n138 = Saturate(None=@n122)
  n139 = OneMinus(None=@n138)
  n140 = TextureSample(UVs=@n32.Rotated Values, Tex=@n45) {samplerType=SAMPLERTYPE_Masks}
  n141 = Multiply(
   A=ComponentMask(None=VectorParameter() {group=Wrap_StarAces, parameterName=SAces_StarsB_Speed, defaultValue=<rgba 0.1 -0.03 0 1>}.RGB) {g=true, r=true}
   B=Time()
  )
  n142 = MaterialFunctionCall(Color=@n7.RGB) {materialFunction=&SplitComponents}
  n143 = LinearInterpolate(
   A=ScalarParameter() {defaultValue=2, group=Wrap_StarAces, parameterName=SAces_StarsB_TileRate_Weapon}
   B=LinearInterpolate(
    A=ScalarParameter() {defaultValue=2, group=Wrap_StarAces, parameterName=SAces_StarsB_TileRate_Vehicle}
    B=ScalarParameter() {defaultValue=2, group=Wrap_StarAces, parameterName=SAces_StarsB_TileRate_VehicleSide}
    Alpha=@n142.G
   )
   Alpha=@n142.R
  )
  n144 = ScalarParameter() {defaultValue=2, group=Wrap_StarAces, parameterName=SAces_StarsB_TileRate_Mech}
  n145 = TextureObjectParameter() {group=Wrap_StarAces, parameterName=SAces_Stars__Text_Mask, texture=&T_BlacksRGB}
  n146 = TextureSample(UVs=Multiply(A=Add(A=@n32.Rotated Values, B=@n141), B=LinearInterpolate(A=@n143, B=@n144, Alpha=@n142.B)), Tex=@n145) {samplerType=SAMPLERTYPE_Masks}
  n147 = ScalarParameter() {defaultValue=1, group=Wrap_StarAces, parameterName=SAces_StarsB_Sparkle_On_State}
  n148 = ScalarParameter() {defaultValue=-1, group=Wrap_StarAces, parameterName=SAces_StarsB_Sparkle_Off_State}
  n149 = ScalarParameter() {defaultValue=0.25, group=Wrap_StarAces, parameterName=SAces_StarsB_Sparkle__Speed}
  n150 = Multiply(A=Time(), B=@n149) {constB=0.1}
  n151 = Frac(None=Add(A=Frac(None=@n150), B=LinearInterpolate(Alpha=@n146.R) {constA=-0.5, constB=1.5}))
  n152 = Multiply(
   A=MaterialFunctionCall(Transition Phase=@n140.R, Height Texture=@n146.A, Contrast=Constant() {r=5}) {materialFunction=&HeightLerp}.Alpha
   B=Saturate(None=LinearInterpolate(A=@n147, B=@n148, Alpha=@n151))
  )
  n153 = MaterialFunctionCall(Color=Abs(None=@n27)) {materialFunction=&SplitComponents}
  n154 = MaterialFunctionCall(Color=@n7.RGB) {materialFunction=&SplitComponents}
  n155 = LinearInterpolate(A=@n153.B, B=@n153.B, Alpha=@n154.G)
  n156 = ScalarParameter() {defaultValue=0.5, group=Wrap_StarAces, parameterName=SAces_Tri_Planner_MaskContrast, sortPriority=5}
  n157 = MaterialFunctionCall(In=LinearInterpolate(A=LinearInterpolate(A=@n153.R, B=@n155, Alpha=@n154.R), B=@n153.G, Alpha=@n154.B), Contrast=@n156) {materialFunction=&CheapContrast}
  n158 = LinearInterpolate(
   A=Multiply(A=LinearInterpolate(A=@n124.RGB, B=@n125.RGB, Alpha=@n128), B=Power(Base=@n137.Raw_Data))
   B=VectorParameter() {group=Wrap_StarAces, parameterName=SAces_StarsB_Color, defaultValue=<rgba 0.75 0.4258 0 1>}.RGB
   Alpha=Multiply(A=@n157.Result, B=Multiply(A=@n139, B=@n152))
  )
  n159 = VectorParameter() {group=Wrap_StarAces, parameterName=SAces_StarsA_Color, defaultValue=<rgba 1 0.5678 0 1>}
  n160 = Multiply(
   A=ComponentMask(None=VectorParameter() {group=Wrap_StarAces, parameterName=SAces_StarsA_Speed, defaultValue=<rgba 0.05 -0.015 0 1>}.RGB) {g=true, r=true}
   B=Time()
  )
  n161 = MaterialFunctionCall(Color=@n7.RGB) {materialFunction=&SplitComponents}
  n162 = LinearInterpolate(
   A=ScalarParameter() {defaultValue=1, group=Wrap_StarAces, parameterName=SAces_StarsA_TileRate_Weapon}
   B=LinearInterpolate(
    A=ScalarParameter() {defaultValue=1, group=Wrap_StarAces, parameterName=SAces_StarsA_TileRate_Vehicle}
    B=ScalarParameter() {defaultValue=1, group=Wrap_StarAces, parameterName=SAces_StarsA_TileRate_VehicleSide}
    Alpha=@n161.G
   )
   Alpha=@n161.R
  )
  n163 = ScalarParameter() {defaultValue=1, group=Wrap_StarAces, parameterName=SAces_StarsA_TileRate_Mech}
  n164 = TextureSample(UVs=Multiply(A=Add(A=@n32.Rotated Values, B=@n160), B=LinearInterpolate(A=@n162, B=@n163, Alpha=@n161.B)), Tex=@n145) {samplerType=SAMPLERTYPE_Masks}
  n165 = MaterialFunctionCall(Transition Phase=@n140.R, Height Texture=@n164.A, Contrast=Constant() {r=5}) {materialFunction=&HeightLerp}
  n166 = Time()
  n167 = ScalarParameter() {defaultValue=0.25, group=Wrap_StarAces, parameterName=SAces_StarsA_Sparkle__Speed}
  n168 = Add(A=Frac(None=Multiply(A=@n166, B=@n167) {constB=0.1}), B=LinearInterpolate(Alpha=@n164.R) {constA=-0.25, constB=1.25})
  n169 = Saturate(None=LinearInterpolate(
    A=ScalarParameter() {defaultValue=1, group=Wrap_StarAces, parameterName=SAces_StarsA_Sparkle_On_State}
    B=ScalarParameter() {defaultValue=-1, group=Wrap_StarAces, parameterName=SAces_StarsA_Sparkle_Off_State}
    Alpha=Frac(None=@n168)
   ))
  n170 = Multiply(A=@n157.Result, B=Multiply(A=@n139, B=Multiply(A=@n165.Alpha, B=@n169)))
  n171 = TextureObjectParameter() {group=Wrap_StarAces, parameterName=SAces_Card___text, texture=&T_BlacksRGB}
  n172 = TextureSample(UVs=@n120.Rotated Values, Tex=@n171)
  n173 = TextureSample(UVs=@n110.Rotated Values, Tex=@n171)
  n174 = TextureSample(UVs=@n90.Rotated Values, Tex=@n171)
  n175 = LinearInterpolate(A=LinearInterpolate(A=Multiply(A=@n121.A, B=@n172.RGB), B=@n173.RGB, Alpha=@n111.A), B=@n174.RGB, Alpha=@n91.A)
  n176 = TextureSample(UVs=@n100.Rotated Values, Tex=@n171)
  n177 = Multiply(A=@n157.Result, B=@n138)
  n178 = Saturate(None=Multiply(A=Add(A=@n177) {constB=-0.5}) {constB=5})
  n179 = SetMaterialAttributes!(
   MaterialAttributes=@n123
   BaseColor=LinearInterpolate(A=LinearInterpolate(A=@n158, B=@n159.RGB, Alpha=@n170), B=LinearInterpolate(A=@n175, B=@n176.RGB, Alpha=@n101.A), Alpha=@n177)
   Roughness=LinearInterpolate(
    A=ScalarParameter() {defaultValue=0.5, group=Wrap_StarAces, parameterName=SAces_Roghness, sortPriority=0}
    B=ScalarParameter() {defaultValue=0.5, group=Wrap_StarAces, parameterName=SAces_Roghness_Cards, sortPriority=0}
    Alpha=@n178
   )
   Specular=LinearInterpolate(
    A=ScalarParameter() {defaultValue=0.5, group=Wrap_StarAces, parameterName=SAces_Specular, sortPriority=0}
    B=ScalarParameter() {defaultValue=0.5, group=Wrap_StarAces, parameterName=SAces_Specular_Cards, sortPriority=0}
    Alpha=@n178
   )
   EmissiveColor
   Metallic=ScalarParameter() {group=Wrap_StarAces, parameterName=SAces_Metallic, sortPriority=0}
  ) {attributeSetTypes [
    #1
    #2
    #3
    #4
    #5
   ]}
  n180 = FunctionInput() {bUsePreviewValueAsDefault=true, inputName="Customization Mask (_CM B)", inputType=FunctionInput_Scalar, sortPriority=24}
  n181 = BlendMaterialAttributes(A=@n123, B=@n179, Alpha=@n180) {vertexAttributeBlendType=UseA}
  n182 = VectorParameter() {group=Wrap_StarAces, parameterName=SAces_2nd_BaseColor, sortPriority=0, defaultValue=<rgba 0 0.9062 0.1038 1>}
  n183 = SetMaterialAttributes!(
   MaterialAttributes=@n123
   BaseColor=@n182.RGB
   Roughness=ScalarParameter() {defaultValue=0.5, group=Wrap_StarAces, parameterName=SAces_2nd_Roghness, sortPriority=0}
   Specular=ScalarParameter() {defaultValue=0.5, group=Wrap_StarAces, parameterName=SAces_2nd_Specular, sortPriority=0}
   Metallic=ScalarParameter() {group=Wrap_StarAces, parameterName=SAces_2nd_Metallic, sortPriority=0}
  ) {attributeSetTypes [
    #1
    #2
    #3
    #5
   ]}
  n184 = FunctionInput() {bUsePreviewValueAsDefault=true, inputName="Customization Mask (_CM B)_1", inputType=FunctionInput_Scalar, sortPriority=24}
  FunctionOutput(None=StaticSwitchParameter(True=BlendMaterialAttributes(A=@n181, B=@n183, Alpha=@n184) {vertexAttributeBlendType=UseA}, False=@n123) {group=Customization_Wrap, parameterName=Wrap_StarAces})
  n185 = ScalarParameter() {defaultValue=0.5, group=Wrap_StarAces, parameterName=SAces_Tri_Planner_MaskContrast_Vehicle, sortPriority=5}
  n186 = ScalarParameter() {defaultValue=0.5, group=Wrap_StarAces, parameterName=SAces_Tri_Planner_MaskContrast_VehicleSide, sortPriority=5}
  n187 = MaterialFunctionCall(Color=@n7.RGB) {materialFunction=&SplitComponents}
  LinearInterpolate(
   A=LinearInterpolate(A=@n156, B=LinearInterpolate(A=@n185, B=@n186, Alpha=@n187.G), Alpha=@n187.R)
   B=ScalarParameter() {defaultValue=0.5, group=Wrap_StarAces, parameterName=SAces_Tri_Planner_MaskContrast_Mech, sortPriority=5}
   Alpha=@n187.B
  )
  n188 = VectorParameter() {group=Wrap_StarAces, parameterName=SAces_StarsA_Speed_Vehicle, defaultValue=<rgba 0 0 0 1>}
  n189 = VectorParameter() {group=Wrap_StarAces, parameterName=SAces_StarsA_Speed_VehicleSide, defaultValue=<rgba 0 0 0 1>}
  n190 = MaterialFunctionCall(Color=@n7.RGB) {materialFunction=&SplitComponents}
  LinearInterpolate(
   A=LinearInterpolate(
    A=VectorParameter() {group=Wrap_StarAces, parameterName=SAces_StarsA_Speed_Weapon, defaultValue=<rgba 0 0 0 1>}.RGB
    B=LinearInterpolate(A=@n188.RGB, B=@n189.RGB, Alpha=@n190.G)
    Alpha=@n190.R
   )
   B=VectorParameter() {group=Wrap_StarAces, parameterName=SAces_StarsA_Speed_Mech, defaultValue=<rgba 0 0 0 1>}.RGB
   Alpha=@n190.B
  )
  n191 = TextureSample(UVs=@n78.Rotated Values, Tex=@n171)
  n192 = TextureSample(UVs=@n68.Rotated Values, Tex=@n171)
  n193 = LinearInterpolate(A=Multiply(A=@n79.A, B=@n191.RGB), B=@n192.RGB, Alpha=@n69.A)
  n194 = TextureSample(UVs=@n44.Rotated Values, Tex=@n171)
  n195 = TextureSample(UVs=@n57.Rotated Values, Tex=@n171)
  LinearInterpolate(A=LinearInterpolate(A=LinearInterpolate(A=@n193, B=@n194.RGB, Alpha=@n46.A), B=@n195.RGB, Alpha=@n58.A), B=@n172.RGB, Alpha=@n121.A)
  n196 = Constant() {r=0.25}
  n197 = MaterialFunctionCall(Float2=@n100.Rotated Values) {materialFunction=&BreakOutFloat2Components}
  n198 = Constant() {r=0.25}
  Multiply(
   A=Multiply(
    A=SphereMask(A=@n196, B=@n197.R, Radius=@n198) {attenuationRadius=0.5}
    B=SphereMask(A=@n197.G, B=@n196, Radius=@n198) {attenuationRadius=0.5}
   )
   B=@n176.A
  )
  n199 = VectorParameter() {parameterName=SAces_StarsA_Speed, defaultValue=<rgba 0.05 -0.015 0 1>}
  n200 = Add(A=@n32.Rotated Values, B=Multiply(A=ComponentMask(None=@n199.RGB) {g=true, r=true}, B=Time())) {constB=0}
  n201 = ScalarParameter() {defaultValue=5, group=Wrap_StarAces, parameterName=SAces_StarsA_TileRate}
  n202 = ComponentMask(None=@n200) {r=true}
  n203 = Divide(A=Floor(None=Multiply(A=@n202, B=@n201) {constB=5}), B=@n201) {constB=5}
  n204 = ScalarParameter() {defaultValue=4, group=Wrap_StarAces, parameterName=SAces_StarsA_Offset_Y}
  n205 = Add(A=Multiply(A=@n200, B=@n201), B=AppendVector(A=Constant(), B=Multiply(A=@n203, B=@n204))) {constB=0}
  n206 = MaterialFunctionCall(UVs=Frac(None=@n205), Texture Scale=ComponentMask(None=@n140.R) {r=true}) {materialFunction=&ScaleUVsByCenter}
  n207 = VectorParameter() {parameterName=SAces_StarsA_Random_seed, defaultValue=<rgba 4 3 0 0>}
  n208 = MaterialFunctionCall!(2 Vector=Add(A=Floor(None=@n205), B=ComponentMask(None=@n207.RGB) {g=true, r=true}), RandomSeed) {materialFunction=&/MF_Random_2}
  n209 = Multiply(
   A=Time()
   B=LinearInterpolate(
    A=ScalarParameter() {defaultValue=-0.1, parameterName=SAces_StarsA_Rotation_Speed_Range_A}
    B=ScalarParameter() {defaultValue=0.15, parameterName=SAces_StarsA_Rotation_Speed_Range_B}
    Alpha=@n208.utput
   )
  )
  n210 = ScalarParameter() {defaultValue=256, parameterName=SAces_Stars_Miplevel _Size}
  Multiply(
   A=TextureSample!(
    UVs=MaterialFunctionCall(UVs=@n206.UVs, "Rotation Angle (0-1)"=@n209) {materialFunction=&CustomRotator}.Rotated Values
    Tex=@n145
    Level=MaterialFunctionCall!(UVs=@n32.Rotated Values, Texture Size=@n210) {materialFunction=/Engine/Functions/Engine_MaterialFunctions01/Texturing/ComputeMipLevel}.MipLevel
    Apply View MipBias
   ) {mipValueMode=TMVM_MipLevel, samplerType=SAMPLERTYPE_Masks}.R
   B=Saturate(None=Round(None=@n208.utput))
  )
  n211 = Constant() {r=0.75}
  n212 = MaterialFunctionCall(Float2=@n120.Rotated Values) {materialFunction=&BreakOutFloat2Components}
  n213 = Constant() {r=0.25}
  n214 = Constant() {r=0.75}
  Multiply(
   A=Multiply(
    A=SphereMask(A=@n211, B=@n212.R, Radius=@n213) {attenuationRadius=0.5}
    B=SphereMask(A=@n212.G, B=@n214, Radius=@n213) {attenuationRadius=0.5}
   )
   B=@n172.A
  )
  n215 = Constant() {r=0.25}
  n216 = MaterialFunctionCall(Float2=@n110.Rotated Values) {materialFunction=&BreakOutFloat2Components}
  n217 = Constant() {r=0.25}
  n218 = Constant() {r=0.75}
  Multiply(
   A=Multiply(
    A=SphereMask(A=@n215, B=@n216.R, Radius=@n217) {attenuationRadius=0.5}
    B=SphereMask(A=@n216.G, B=@n218, Radius=@n217) {attenuationRadius=0.5}
   )
   B=@n173.A
  )
  n219 = VectorParameter() {parameterName=SAces_StarsB_Speed, defaultValue=<rgba 0.1 -0.03 0 1>}
  n220 = Add(A=@n32.Rotated Values, B=Multiply(A=ComponentMask(None=@n219.RGB) {g=true, r=true}, B=Time())) {constB=0}
  n221 = ScalarParameter() {defaultValue=10, group=Wrap_StarAces, parameterName=SAces_StarsB_TileRate}
  n222 = ComponentMask(None=@n220) {r=true}
  n223 = Divide(A=Floor(None=Multiply(A=@n222, B=@n221) {constB=5}), B=@n221) {constB=5}
  n224 = ScalarParameter() {defaultValue=4, group=Wrap_StarAces, parameterName=SAces_StarsB_Offset_Y}
  n225 = Add(A=Multiply(A=@n220, B=@n221), B=AppendVector(A=Constant(), B=Multiply(A=@n223, B=@n224))) {constB=0}
  n226 = MaterialFunctionCall(UVs=Frac(None=@n225), Texture Scale=ComponentMask(None=@n140.R) {r=true}) {materialFunction=&ScaleUVsByCenter}
  n227 = VectorParameter() {parameterName=SAces_StarsB_Random_seed, defaultValue=<rgba 3 5 0 0>}
  n228 = MaterialFunctionCall(2 Vector=Add(A=Floor(None=@n225), B=ComponentMask(None=@n227.RGB) {g=true, r=true})) {materialFunction=&MF_Random_2}
  n229 = Multiply(
   A=Time()
   B=LinearInterpolate(
    A=ScalarParameter() {parameterName=SAces_StarsB_Rotation_Speed_Range_A}
    B=ScalarParameter() {defaultValue=0.3, parameterName=SAces_StarsB_Rotation_Speed_Range_B}
    Alpha=@n228.utput
   )
  )
  Multiply(
   A=TextureSample(
    UVs=MaterialFunctionCall(UVs=@n226.UVs, "Rotation Angle (0-1)"=@n229) {materialFunction=&CustomRotator}.Rotated Values
    Tex=@n145
    Level=MaterialFunctionCall(UVs=@n32.Rotated Values, Texture Size=@n210) {materialFunction=&ComputeMipLevel}.MipLevel
   ) {mipValueMode=TMVM_MipLevel, samplerType=SAMPLERTYPE_Masks}.R
   B=Saturate(None=Round(None=@n228.utput))
  )
  Multiply(B=@n195.A)
  n230 = Constant() {r=0.75}
  n231 = MaterialFunctionCall(Float2=@n44.Rotated Values) {materialFunction=&BreakOutFloat2Components}
  n232 = Constant() {r=0.25}
  n233 = Constant() {r=0.25}
  Multiply(
   A=Multiply(
    A=SphereMask(A=@n230, B=@n231.R, Radius=@n232) {attenuationRadius=0.5}
    B=SphereMask(A=@n231.G, B=@n233, Radius=@n232) {attenuationRadius=0.5}
   )
   B=@n194.A
  )
  n234 = Constant() {r=0.75}
  n235 = MaterialFunctionCall(Float2=@n78.Rotated Values) {materialFunction=&BreakOutFloat2Components}
  n236 = Constant() {r=0.25}
  n237 = Constant() {r=0.75}
  Multiply(
   A=Multiply(
    A=SphereMask(A=@n234, B=@n235.R, Radius=@n236) {attenuationRadius=0.5}
    B=SphereMask(A=@n235.G, B=@n237, Radius=@n236) {attenuationRadius=0.5}
   )
   B=@n191.A
  )
  n238 = Constant() {r=0.25}
  n239 = MaterialFunctionCall(Float2=@n68.Rotated Values) {materialFunction=&BreakOutFloat2Components}
  n240 = Constant() {r=0.25}
  n241 = Constant() {r=0.75}
  Multiply(
   A=Multiply(
    A=SphereMask(A=@n238, B=@n239.R, Radius=@n240) {attenuationRadius=0.5}
    B=SphereMask(A=@n239.G, B=@n241, Radius=@n240) {attenuationRadius=0.5}
   )
   B=@n192.A
  )
  n242 = Constant() {r=0.75}
  n243 = MaterialFunctionCall(Float2=@n90.Rotated Values) {materialFunction=&BreakOutFloat2Components}
  n244 = Constant() {r=0.25}
  n245 = Constant() {r=0.25}
  Multiply(
   A=Multiply(
    A=SphereMask(A=@n242, B=@n243.R, Radius=@n244) {attenuationRadius=0.5}
    B=SphereMask(A=@n243.G, B=@n245, Radius=@n244) {attenuationRadius=0.5}
   )
   B=@n174.A
  )
  OneMinus(None=@n151)
  Saturate(None=Add(A=@n180, B=ScalarParameter() {group=Wrap_StarAces, parameterName=SAces_Remove_Customization Mask, sortPriority=-10}))
  Saturate()
  SphereMask(B=Constant() {r=0.25}, Radius=Constant() {r=0.25}) {attenuationRadius=0.5}
  n246 = Time()
  TextureSample!(UVs=Add(A=@n146.R, B=Multiply(A=@n246, B=@n149)), Tex=@n45, Apply View MipBias) {samplerType=SAMPLERTYPE_Masks}
  n247 = Time()
  TextureSample(UVs=Add(A=@n164.R, B=Multiply(A=@n247, B=@n167)), Tex=@n45) {samplerType=SAMPLERTYPE_Masks}
  VectorParameter() {group=Wrap_StarAces, parameterName=SAces_BaseColor, sortPriority=0, defaultValue=<rgba 0.003 0.003 0.003 1>}
 ]}
