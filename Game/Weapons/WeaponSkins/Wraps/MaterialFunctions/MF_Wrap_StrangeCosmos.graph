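# Strange Cosmos wrap layer: paints animated space over customization-masked areas of weapons, vehicles, mechs, characters and backpacks, behind a static switch.
# State inputs select tiling and offsets for side projection from pre-skinned bounds and normals.
# Gas cloud, two triplanar star layers lit by panning warped gradients, and space elements build color, roughness, specular and emissive, with normal-based side masking and noise breakup.
# Gas-cloud pan/warp and star-scroll nodes are unconnected.
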MaterialFunction /Game/Weapons/WeaponSkins/Wraps/MaterialFunctions/MF_Wrap_StrangeCosmos
graph {nodes [
  n1 = FunctionInput() {inputName=isState, inputType=FunctionInput_Vector4, sortPriority=5}
  n2 = MaterialFunctionCall(Color=@n1) {materialFunction=&SplitComponents}
  n3 = VectorParameter() {group=Wrap_StrangCosmos_UVs, parameterName=SC_Offset_Weapon, defaultValue=<rgba 0 0 0 1>}
  n4 = MaterialFunctionCall(Float4=@n1) {materialFunction=&BreakOutFloat4Components}
  n5 = LinearInterpolate(
   A=VectorParameter() {group=Wrap_StrangCosmos_UVs, parameterName=SC_Offset_Vehicle, defaultValue=<rgba 0 0 0 1>}.RGB
   B=VectorParameter() {group=Wrap_StrangCosmos_UVs, parameterName=SC_Offset_Vehicle_Side, defaultValue=<rgba 0 0 0 1>}.RGB
   Alpha=@n4.G
  )
  n6 = VectorParameter() {group=Wrap_StrangCosmos_UVs, parameterName=SC_Offset_Mech -----, defaultValue=<rgba 0 0 0 1>}
  n7 = VectorParameter() {group=Wrap_StrangCosmos_UVs, parameterName=SC_Offset_Character, defaultValue=<rgba 0 0 0 1>}
  n8 = ScalarParameter() {defaultValue=1, group=Wrap_StrangCosmos_UVs, parameterName=SC_TileRate_Weapon}
  n9 = LinearInterpolate(
   A=ScalarParameter() {defaultValue=1, group=Wrap_StrangCosmos_UVs, parameterName=SC_TileRate_Vehicle}
   B=ScalarParameter() {defaultValue=1, group=Wrap_StrangCosmos_UVs, parameterName=SC_TileRate_VehicleSide}
   Alpha=@n4.G
  )
  n10 = ScalarParameter() {defaultValue=1, group=Wrap_StrangCosmos_UVs, parameterName=SC_TileRate_Mech -----}
  n11 = LinearInterpolate(
   A=ScalarParameter() {defaultValue=1, group=Wrap_StrangCosmos_UVs, parameterName=SC_TileRate_Character}
   B=ScalarParameter() {defaultValue=1, group=Wrap_StrangCosmos_UVs, parameterName=SC_TileRate_Backpack}
   Alpha=NamedRerouteUsage()
  )
  n12 = VertexInterpolator(VS=PreSkinnedPosition())
  n13 = PreSkinnedLocalBounds()
  n14 = FunctionInput(Preview=Divide(A=Subtract(A=@n12.PS, B=@n13.Min), B=Subtract(A=@n13.Max, B=@n13.Max))) {bUsePreviewValueAsDefault=true, inputName=Calcated_Bounds, sortPriority=11}
  n15 = FunctionInput(Preview=VertexInterpolator(VS=PreSkinnedNormal()).PS) {bUsePreviewValueAsDefault=true, inputName=Pre-Skinned Local Normal, sortPriority=10}
  n16 = MaterialFunctionCall!(
   For_Safe_Zones
   Normal_Map
   Caculated_Bounds=@n14
   Pre_Skinned_Normals=@n15
   Is Vehicle=@n2.R
   Is VehicleSide=@n2.R
   Is Mech=@n2.R
   Vehicle_Rotation=ScalarParameter() {group=Wrap_StrangCosmos_UVs, parameterName=SC_Rotation_Vehicle}
   UV_Rotation Bypass
   Mech_Rotation=ScalarParameter() {defaultValue=0.25, group=Wrap_StrangCosmos_UVs, parameterName=SC_Rotation_Mech -----}
   Uvs_Offset=LinearInterpolate(A=LinearInterpolate(A=@n3.RGB, B=@n5, Alpha=@n4.R), B=LinearInterpolate(A=@n6.RGB, B=@n7.RGB, Alpha=@n4.A), Alpha=@n4.B)
   Offset-after rotation
   Uvs_TileRate=LinearInterpolate(A=LinearInterpolate(A=@n8, B=@n9, Alpha=@n4.R), B=LinearInterpolate(A=@n10, B=@n11, Alpha=@n4.A), Alpha=@n4.B)
  ) {materialFunction=&/MF_Projection_Side_Normal}
  n17 = VectorParameter() {group=Wrap_StrangCosmos_GasCloud, parameterName=SC_GasCloud_Speed_Weapon, defaultValue=<rgba 0.025 0 0 1>}
  n18 = VectorParameter() {group=Wrap_StrangCosmos_GasCloud, parameterName=SC_GasCloud_Speed_Vehicle, defaultValue=<rgba 0.1 0.1 0.1 1>}
  n19 = VectorParameter() {group=Wrap_StrangCosmos_GasCloud, parameterName=SC_GasCloud_Speed_VehicleSide, defaultValue=<rgba 0.1 0.1 0.1 1>}
  n20 = MaterialFunctionCall(Color=@n1) {materialFunction=&SplitComponents}
  n21 = LinearInterpolate(A=@n18.RGB, B=@n19.RGB, Alpha=@n20.G)
  n22 = VectorParameter() {group=Wrap_StrangCosmos_GasCloud, parameterName=SC_GasCloud_Speed_Mech-----, defaultValue=<rgba 0.1 0.1 0.1 1>}
  n23 = Multiply(A=Time(), B=LinearInterpolate(A=LinearInterpolate(A=@n17.RGB, B=@n21, Alpha=@n20.R), B=@n22.RGB, Alpha=@n20.B))
  n24 = Multiply(A=@n16.UVs, B=ScalarParameter() {defaultValue=1, group=Wrap_StrangCosmos_GasCloud, parameterName=SC_GasCloud_Warp_TileRate, sortPriority=50})
  n25 = ComponentMask(None=Multiply(A=Time(), B=ScalarParameter() {defaultValue=0.015, group=Wrap_StrangCosmos_GasCloud, parameterName=SC_GasCloud_Warp_Speed, sortPriority=50})) {g=true, r=true}
  n26 = TextureObjectParameter() {group=Wrap__StrangeCosmos_Textures, parameterName=SC_Mask_01, sortPriority=20, texture=&T_BlacksRGB}
  n27 = ChannelMaskParameter(None=TextureSample(UVs=Add(A=@n24, B=@n25), Tex=@n26) {samplerType=SAMPLERTYPE_LinearGrayscale}.RGBA) {group=Wrap_StrangCosmos_GasCloud, maskChannel=Blue, parameterName=SC_GasCloud_Warp__Channel, sortPriority=50, defaultValue=<rgba 0 0 1 0>}
  Add(
   A=Add(A=@n16.UVs, B=ComponentMask(None=@n23) {g=true, r=true})
   B=Multiply(
    A=@n27.R
    B=ScalarParameter() {defaultValue=0.035, group=Wrap_StrangCosmos_GasCloud, parameterName=SC_GasCloud_Warp_Amount -----, sortPriority=50}
   )
  )
  n28 = TextureObjectParameter() {group=Wrap__StrangeCosmos_Textures, parameterName=SC_Mask_02, sortPriority=20, texture=&T_BlacksRGB}
  n29 = TextureSample(UVs=@n16.UVs, Tex=@n28)
  n30 = VectorParameter() {group=Wrap_StrangeCosmos_Masking, parameterName="SC_Mask_BaseColor Rough=A", sortPriority=0, defaultValue=<rgba 0.0041 0.0041 0.009 0.5>}
  n31 = MaterialFunctionCall(Color=Abs(None=@n15)) {materialFunction=&SplitComponents}
  n32 = MaterialFunctionCall(Color=@n1) {materialFunction=&SplitComponents}
  n33 = LinearInterpolate(A=Add(A=@n31.R, B=@n31.R), B=Add(A=@n31.B, B=@n31.B), Alpha=@n32.R)
  n34 = Add(A=@n31.R, B=@n31.R)
  n35 = ScalarParameter() {defaultValue=0.1, group=Wrap_StrangeCosmos_Masking, parameterName=SC_Mask_Contrast, sortPriority=51}
  n36 = OneMinus(None=MaterialFunctionCall(In=LinearInterpolate(A=@n33, B=@n34, Alpha=@n32.B), Contrast=@n35) {materialFunction=&CheapContrast}.Result)
  n37 = MaterialFunctionCall(Color=Abs(None=@n15)) {materialFunction=&SplitComponents}
  n38 = LinearInterpolate(A=@n37.R, B=LinearInterpolate(A=@n37.B, B=@n37.B, Alpha=@n32.G), Alpha=@n32.R)
  n39 = OneMinus(None=MaterialFunctionCall(In=LinearInterpolate(A=@n38, B=@n37.G, Alpha=@n32.B), Contrast=@n35) {materialFunction=&CheapContrast}.Result)
  n40 = ScalarParameter() {defaultValue=1, group=Wrap_StrangeCosmos_Masking, parameterName=SC_Mask_Z+AltAxisActive -----, sortPriority=50}
  n41 = LinearInterpolate(
   A=LinearInterpolate(
    A=Constant()
    B=Saturate(None=@n36)
    Alpha=ScalarParameter() {defaultValue=1, group=Wrap_StrangeCosmos_Masking, parameterName=SC_Mask_AltAxisActive -----, sortPriority=50}
   )
   B=LinearInterpolate(A=Constant(), B=Saturate(None=@n39), Alpha=@n40)
   Alpha=@n40
  )
  n42 = LinearInterpolate(A=ComponentMask(None=@n29.RGB) {b=true, g=true, r=true}, B=@n30.RGB, Alpha=@n41)
  n43 = VectorParameter() {group=Wrap_StrangeCosmos, parameterName="SC_Stars01_BaseColor A=Roughness", sortPriority=-26, defaultValue=<rgba 0.8388 0.7835 0.6038 0.2>}
  n44 = Multiply(A=@n29.A, B=OneMinus(None=@n41))
  n45 = OneMinus(None=@n44)
  n46 = Time()
  n47 = ScalarParameter() {defaultValue=0.015, group=Wrap_StrangeCosmos, parameterName=SC_Gradient_Speed}
  n48 = Multiply(
   A=Add(A=Multiply(A=@n46, B=@n47), B=@n16.UVs)
   B=ScalarParameter() {defaultValue=1, group=Wrap_StrangeCosmos, parameterName=SC_Gradient_TileRate}
  )
  n49 = StaticBool() {value=true}
  n50 = TextureObjectParameter() {group=Wrap__StrangeCosmos_Textures, parameterName=SC_Mask_03, sortPriority=20, texture=&T_BlacksRGB}
  n51 = Constant() {r=1}
  n52 = MaterialFunctionCall(Color=Constant3Vector() {constant=<rgba 0 -0.25 -0.25 1>}) {materialFunction=&SplitComponents}
  n53 = ScalarParameter() {defaultValue=1, group=Wrap_StrangeCosmos, parameterName=SC_HFNoise_Blend_Contrast, sortPriority=-10}
  n54 = MaterialFunctionCall(
   Export Float 4=@n49
   TextureObject=@n50
   TextureSize=@n51
   WorldPosition=@n14
   World Space Normal=@n15
   RotZAxis=@n52.B
   RotXAxis=@n52.B
   RotYAxis=@n52.B
   ProjectionTransitionContrast=@n53
  ) {materialFunction=&/WorldAlignedTexture}
  n55 = ChannelMaskParameter(None=@n54.XYZ Texture) {group=Wrap_StrangeCosmos, maskChannel=Alpha, parameterName=SC_Stars02_WarpMask_01, sortPriority=-26, defaultValue=<rgba 0 0 0 1>}
  n56 = Multiply(A=@n55.R, B=ScalarParameter() {defaultValue=1, group=Wrap_StrangeCosmos, parameterName=SC_Stars02_WarpAmount, sortPriority=-26})
  n57 = ChannelMaskParameter(None=TextureSample(UVs=Add(A=@n48, B=@n56), Tex=@n26).RGBA) {group=Wrap_StrangeCosmos, maskChannel=Blue, parameterName=SC_Gradient_ChannelMask_01, defaultValue=<rgba 0 0 1 0>}
  n58 = ChannelMaskParameter(None=@n54.XYZ Texture) {group=Wrap_StrangeCosmos, maskChannel=Blue, parameterName=SC_Stars02_ChannelMask_01, sortPriority=-26, defaultValue=<rgba 0 0 1 0>}
  n59 = Multiply(A=@n57.R, B=@n58.R)
  n60 = Add(A=@n59, B=@n58.R)
  n61 = Round(None=Saturate(None=Multiply(A=@n59, B=@n60)))
  n62 = Multiply(A=@n45, B=@n61)
  n63 = VectorParameter() {group=Wrap_StrangeCosmos, parameterName="SC_Stars02_BaseColor A=Roughness", sortPriority=-26, defaultValue=<rgba 0.8228 0.7991 0.0003 0.2>}
  n64 = ChannelMaskParameter(None=@n54.XYZ Texture) {group=Wrap_StrangeCosmos, maskChannel=Green, parameterName=SC_Stars01_WarpMask_01, sortPriority=-26, defaultValue=<rgba 0 1 0 0>}
  n65 = Multiply(A=@n64.R, B=ScalarParameter() {defaultValue=1, group=Wrap_StrangeCosmos, parameterName=SC_Stars01_WarpAmount, sortPriority=-26})
  n66 = ChannelMaskParameter(None=TextureSample(UVs=Add(A=@n48, B=@n65), Tex=@n26).RGBA) {group=Wrap_StrangeCosmos, maskChannel=Blue, parameterName=SC_Gradient_ChannelMask_01, defaultValue=<rgba 0 0 1 0>}
  n67 = ChannelMaskParameter(None=@n54.XYZ Texture) {group=Wrap_StrangeCosmos, parameterName=SC_Stars01_ChannelMask_01, sortPriority=-26}
  n68 = Multiply(A=@n66.R, B=@n67.R)
  n69 = Add(A=@n68, B=@n67.R)
  n70 = Round(None=Saturate(None=Multiply(A=@n68, B=@n69)))
  n71 = Multiply(A=@n70, B=@n45)
  n72 = TextureObjectParameter() {group=Wrap__StrangeCosmos_Textures, parameterName=SC_BaseColor_01, sortPriority=11, texture=&T_BlacksRGB}
  n73 = TextureSample(UVs=@n16.UVs, Tex=@n72)
  n74 = Multiply(A=@n73.A, B=OneMinus(None=@n41))
  n75 = LinearInterpolate(
   A=LinearInterpolate(A=LinearInterpolate(A=@n42, B=@n43.RGB, Alpha=@n62), B=@n63.RGB, Alpha=@n71)
   B=ComponentMask(None=@n73.RGB) {b=true, g=true, r=true}
   Alpha=@n74
  )
  n76 = Add(A=@n70, B=@n61)
  n77 = TextureSample(UVs=@n16.UVs, Tex=@n26)
  n78 = ChannelMaskParameter(None=@n77.RGBA) {group=Wrap_StrangeCosmos, maskChannel=Green, parameterName=SC_Emissive_WarpMask, defaultValue=<rgba 0 1 0 0>}
  n79 = Time()
  n80 = ScalarParameter() {defaultValue=0.25, group=Wrap_StrangeCosmos, parameterName=SC_Gradient_SpeedEmissive}
  n81 = Multiply(
   A=Add(A=@n78.R, B=Multiply(A=@n79, B=@n80))
   B=ScalarParameter() {defaultValue=2, group=Wrap_StrangeCosmos, parameterName=SC_Emissive_WarpAmount}
  )
  n82 = ChannelMaskParameter(None=TextureSample(UVs=Add(A=@n48, B=@n81), Tex=@n26).RGBA) {group=Wrap_StrangeCosmos, maskChannel=Blue, parameterName=SC_Gradient_ChannelMask_01, defaultValue=<rgba 0 0 1 0>}
  n83 = ChannelMaskParameter(None=@n77.RGBA) {group=Wrap_StrangeCosmos, maskChannel=Alpha, parameterName=SC_Emissive_ChannelMask-----, defaultValue=<rgba 0 0 0 1>}
  n84 = Multiply(A=@n82.R, B=@n83.R)
  n85 = Multiply(A=@n84, B=Add(A=@n84, B=@n83.R))
  n86 = LinearInterpolate(
   A=Multiply(A=Multiply(A=@n76, B=@n45), B=OneMinus(None=@n74))
   B=Multiply(A=Saturate(None=@n85), B=ScalarParameter() {defaultValue=5, group=Wrap_StrangeCosmos, parameterName=SC_Emissive_Intensity})
   Alpha=@n83.R
  )
  n87 = Multiply(A=@n75, B=LinearInterpolate(A=OneMinus(None=@n86), B=@n41, Alpha=@n41) {constAlpha=1})
  n88 = ChannelMaskParameter(None=MaterialFunctionCall(
    Export Float 4=@n49
    TextureObject=@n26
    TextureSize=@n51
    WorldPosition=@n14
    World Space Normal=@n15
    RotZAxis=@n52.B
    RotXAxis=@n52.B
    RotYAxis=@n52.B
    ProjectionTransitionContrast=@n53
   ) {materialFunction=&WorldAlignedTexture}.XYZ Texture) {group=Wrap_StrangeCosmos, parameterName=SC_HFNoise__ChannelMask-----, sortPriority=-10}
  n89 = Multiply(
   A=OneMinus(None=@n88.R)
   B=ScalarParameter() {defaultValue=1.25, group=Wrap_StrangeCosmos, parameterName=SC_HFNoise_MultiplyBlend, sortPriority=-10}
  )
  n90 = Multiply(A=@n88.R, B=ScalarParameter() {defaultValue=0.035, group=Wrap_StrangeCosmos, parameterName=SC_HFNoise_AddBlend, sortPriority=-10})
  n91 = LinearInterpolate(
   A=ScalarParameter() {defaultValue=0.25, group=Wrap_StrangeCosmos, parameterName=SC_Specular, sortPriority=1}
   B=ScalarParameter() {defaultValue=0.2, group=Wrap_StrangeCosmos, parameterName=SC_Specular_GasCloud, sortPriority=0}
   Alpha=@n44
  )
  n92 = ScalarParameter() {group=Wrap_StrangeCosmos, parameterName=SC_Specular_Stars, sortPriority=1}
  n93 = LinearInterpolate(
   A=LinearInterpolate(A=LinearInterpolate(A=@n91, B=@n92, Alpha=@n62), B=@n92, Alpha=@n71)
   B=ScalarParameter() {defaultValue=0.3, group=Wrap_StrangeCosmos, parameterName=SC_Specular_SpaceElements, sortPriority=2}
   Alpha=@n74
  )
  n94 = LinearInterpolate(
   A=@n30.A
   B=ScalarParameter() {defaultValue=0.7, group=Wrap_StrangeCosmos, parameterName=SC_Roughness_GasCloud-----, sortPriority=0}
   Alpha=@n44
  )
  n95 = LinearInterpolate(
   A=LinearInterpolate(A=LinearInterpolate(A=@n94, B=@n43.A, Alpha=@n62), B=@n63.A, Alpha=@n71)
   B=ScalarParameter() {defaultValue=0.5, group=Wrap_StrangeCosmos, parameterName=SC_Roughness_SpaceElements, sortPriority=2}
   Alpha=@n74
  )
  n96 = Constant()
  n97 = GetMaterialAttributes(None=FunctionInput() {inputName=Material, inputType=FunctionInput_MaterialAttributes, sortPriority=-20}) {attributeGetTypes [#1]}
  n98 = SetMaterialAttributes!(
   MaterialAttributes=@n97.MaterialAttributes
   BaseColor=Add(A=Multiply(A=@n87, B=@n89), B=@n90)
   Metallic=ScalarParameter() {group=Wrap_StrangeCosmos, parameterName=SC_Metallic, sortPriority=1}
   Specular=Add(A=Multiply(A=@n93, B=@n89), B=@n90)
   Roughness=Add(A=Multiply(A=@n95, B=@n89), B=@n90)
   EmissiveColor=LinearInterpolate(A=LinearInterpolate(A=@n96, B=@n75, Alpha=@n86), Alpha=@n41) {constB=0}
   Normal=@n97.Normal
  ) {attributeSetTypes [
    #2
    #3
    #4
    #5
    #6
    #1
   ]}
  n99 = FunctionInput() {
   bUsePreviewValueAsDefault=true
   desc=base customization
   inputName="Customization Mask (_CM B Top left)"
   inputType=FunctionInput_Scalar
   sortPriority=49
  }
  n100 = FunctionInput() {
   bUsePreviewValueAsDefault=true
   desc=secondary
   inputName="Customization Mask 2 (_CM B Top Right)"
   inputType=FunctionInput_Scalar
   sortPriority=50
  }
  n101 = Add(A=@n99, B=LinearInterpolate(A=@n100, Alpha=NamedRerouteUsage()) {constB=0})
  FunctionOutput(None=StaticSwitchParameter(True=BlendMaterialAttributes(A=@n97.MaterialAttributes, B=@n98, Alpha=@n101) {vertexAttributeBlendType=UseA}, False=@n97.MaterialAttributes) {group=Customization_Wrap, parameterName=Wrap_StrangeCosmos})
  n102 = VectorParameter() {group=Wrap_StrangCosmos_GasCloud, parameterName=SC_GasCloud_Warp_Speed_Weapon, sortPriority=50, defaultValue=<rgba 0.1 0.1 0.1 1>}
  n103 = VectorParameter() {group=Wrap_StrangCosmos_GasCloud, parameterName=SC_GasCloud_Warp_Speed_Vehicle, sortPriority=50, defaultValue=<rgba 0.1 0.1 0.1 1>}
  n104 = VectorParameter() {group=Wrap_StrangCosmos_GasCloud, parameterName=SC_GasCloud_Warp_Speed_VehicleSide, sortPriority=50, defaultValue=<rgba 0.1 0.1 0.1 1>}
  n105 = MaterialFunctionCall(Color=@n1) {materialFunction=&SplitComponents}
  n106 = VectorParameter() {group=Wrap_StrangCosmos_GasCloud, parameterName=SC_GasCloud_Warp_Speed_Mech, sortPriority=50, defaultValue=<rgba 0.1 0.1 0.1 1>}
  LinearInterpolate(A=LinearInterpolate(A=@n102.RGB, B=LinearInterpolate(A=@n103.RGB, B=@n104.RGB, Alpha=@n105.G), Alpha=@n105.R), B=@n106.RGB, Alpha=@n105.B)
  NamedRerouteDeclaration(None=@n4.A) {name=IsCharacter, nodeColor=<rgba 0.8459 1 0 1>}
  n107 = FunctionInput() {bUsePreviewValueAsDefault=true, inputName=IsStatsB, inputType=FunctionInput_Vector2, sortPriority=78}
  NamedRerouteDeclaration(None=ComponentMask(None=NamedRerouteDeclaration(None=@n107) {name=Is_StatsB, nodeColor=<rgba 0 0.9253 1 1>}) {r=true}) {name=IsBackpack, nodeColor=<rgba 0.2656 0 1 1>}
  n108 = Multiply(A=Time(), B=ScalarParameter() {defaultValue=-0.01, group=Wrap_StrangeCosmos, parameterName=SC_Stars_Speed, sortPriority=-26})
  TextureSample(UVs=Add(A=@n16.UVs, B=ComponentMask(None=@n108) {g=true, r=true}), Tex=@n50)
 ]}
