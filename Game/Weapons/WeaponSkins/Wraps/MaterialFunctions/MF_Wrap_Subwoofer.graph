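# Material function for the Subwoofer wrap on weapons, vehicles and mechs, adding an audio-reactive, glowing equalizer look.
# Pre-skinned bounds drive per-type projections for the mask texture, background gradient and cell-quantized wave coordinates.
# A scrolling sine texture switches to custom-HLSL parabolic bands during emote or event audio, or a Blueprint-timed pulse.
# Curve-colored waves and curvature edges feed emissive, blended by customization mask behind a static switch.
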MaterialFunction /Game/Weapons/WeaponSkins/Wraps/MaterialFunctions/MF_Wrap_Subwoofer
graph {nodes [
  n1 = MaterialFunctionCall(
   X=ScalarParameter() {defaultValue=30, group=Wrap_Subwoofer, parameterName=SubW_CellsX}
   Y=ScalarParameter() {defaultValue=200, group=Wrap_Subwoofer, parameterName=SubW_CellsY}
  ) {materialFunction=&MakeFloat2}
  n2 = PreSkinnedPosition()
  n3 = PreSkinnedLocalBounds()
  n4 = Divide(A=Subtract(A=VertexInterpolator(VS=@n2).PS, B=@n3.Min), B=Subtract(A=@n3.Max, B=@n3.Max))
  n5 = FunctionInput(Preview=@n4) {bUsePreviewValueAsDefault=true, inputName=Calcated_Bounds, sortPriority=11}
  n6 = MaterialFunctionCall(Color=@n5) {materialFunction=&SplitComponents}
  n7 = AppendVector(A=@n6.R, B=@n6.R)
  n8 = ScalarParameter() {group=Wrap_Subwoofer, parameterName=SubW_Rotation_Vehicle, sortPriority=500}
  n9 = AppendVector(A=@n6.R, B=OneMinus(None=@n6.B))
  n10 = Constant() {r=1}
  n11 = FunctionInput(Preview=@n10) {bUsePreviewValueAsDefault=true, inputName=Side Project Vehicle, inputType=FunctionInput_Scalar}
  n12 = FunctionInput(Preview=@n10) {bUsePreviewValueAsDefault=true, inputName=isVehicle, inputType=FunctionInput_Scalar, sortPriority=5}
  n13 = LinearInterpolate(
   A=AppendVector(A=MaterialFunctionCall(Color=@n5) {materialFunction=&SplitComponents}.G, B=OneMinus(None=@n6.B))
   B=LinearInterpolate(
    A=MaterialFunctionCall(UVs=@n7, "Rotation Angle (0-1)"=@n8) {materialFunction=&CustomRotator}.Rotated Values
    B=@n9
    Alpha=@n11
   )
   Alpha=@n12
  )
  n14 = FunctionInput(Preview=Constant()) {bUsePreviewValueAsDefault=true, inputName=isMech, inputType=FunctionInput_Scalar, sortPriority=4}
  n15 = Multiply(A=Frac(None=LinearInterpolate(A=@n13, B=@n9, Alpha=@n14)), B=@n1.Result) {constB=5}
  n16 = Floor(None=ComponentMask(None=@n15) {r=true})
  n17 = ComponentMask(None=@n15) {g=true}
  n18 = MaterialFunctionCall(Float2=Divide(A=AppendVector(A=@n16, B=@n17), B=@n1.Result) {constB=5}) {materialFunction=&BreakOutFloat2Components}
  n19 = AppendVector(A=Multiply(A=@n18.R, B=ScalarParameter() {defaultValue=4, group=Wrap_Subwoofer, parameterName=SubW_U_mul}) {constB=0.5}, B=@n18.G)
  MaterialExpressionMaterialFunctionCall_31 = MaterialFunctionCall(Float2=@n19) {materialFunction=&BreakOutFloat2Components}
  n20 = Custom!(x=@MaterialExpressionMaterialFunctionCall_31.R) {
   code="return abs(2.*frac(x)-1.);"
   description=Mirror
   outputType=CMOT_Float1
   inputs [{
     inputName=x
     input {expression=&self:MaterialExpressionMaterialFunctionCall_31}
    }]
  }
  AppendVector(A=@n20, B=@MaterialExpressionMaterialFunctionCall_31.G)
  n21 = FunctionInput() {bUsePreviewValueAsDefault=true, inputName=Customization Mask Curvature, inputType=FunctionInput_Scalar, sortPriority=24}
  n22 = Constant() {r=0.5}
  n23 = Constant() {r=1}
  n24 = Constant()
  n25 = MaterialFunctionCall(
   In=Multiply(
    A=MaterialFunctionCall(Input=@n21, Input Low=@n22, Input High=@n23, Target Low=@n24, Target High=@n23) {materialFunction=&RemapValueRange}.Result
    B=ScalarParameter() {defaultValue=1, group=Wrap_Subwoofer, parameterName=SubW_Edge_Multplyer}
   )
   Contrast=ScalarParameter() {group=Wrap_Subwoofer, parameterName=SubW_Edge_Contrast}
  ) {materialFunction=&CheapContrast}
  n26 = TextureObjectParameter() {
   group=Wrap_Subwoofer
   parameterName=SubW_mask_texture
   samplerType=SAMPLERTYPE_LinearColor
   sortPriority=50
   texture=/Game/Weapons/WeaponSkins/Wraps/Textures/Subwoofer/T_Wrap_Subwoofer_M_2
  }
  n27 = LinearInterpolate(A=Constant3Vector() {constant=<rgba 0 0 -0.25 1>}, B=Constant3Vector() {constant=<rgba 0 0 0 1>}, Alpha=@n12)
  n28 = Constant3Vector() {constant=<rgba 0.25 0.25 0 1>}
  n29 = VectorParameter() {group=Wrap_Subwoofer, parameterName="SubW_Fill_alternate_Axis_Color_A=Roughness", sortPriority=51}
  n30 = LinearInterpolate(
   A=LinearInterpolate(
    A=ScalarParameter() {defaultValue=50, group=Wrap_Subwoofer, parameterName=SubW_Mask_TileRate_Weapon, sortPriority=54}
    B=ScalarParameter() {defaultValue=50, group=Wrap_Subwoofer, parameterName=SubW_Mask_TileRate_Vehicle, sortPriority=54}
    Alpha=@n12
   )
   B=ScalarParameter() {defaultValue=50, group=Wrap_Subwoofer, parameterName=SubW_Mask_TileRate_Mech, sortPriority=54}
   Alpha=@n14
  )
  n31 = VertexInterpolator(VS=PreSkinnedNormal())
  n32 = MaterialFunctionCall(Float4=MaterialFunctionCall!(
    IsCharacter
    IsStatsB
    Texture=@n26
    Rotation_Axis=LinearInterpolate(A=@n27, B=@n28, Alpha=@n14)
    Caculated_Bounds=Multiply(A=@n30, B=@n5)
    Pre_Skinned_Normals=FunctionInput(Preview=@n31.PS) {bUsePreviewValueAsDefault=true, inputName=Pre-Skinned Local Normal, sortPriority=10}
    Blend_Contrast=ScalarParameter() {defaultValue=1, group=Wrap_Subwoofer, parameterName=SubW_Blend_Contrast, sortPriority=51}
    Fill_alternate_Axis_Opacity=ScalarParameter() {group=Wrap_Subwoofer, parameterName=SubW_Fill_alternate_Axis_Opacity, sortPriority=51}
    "Fill_alternate_Axis_Color_A=Roughness"=AppendVector(A=@n29.RGB, B=@n29.RGB)
    Is_Vehicle=@n12
    Is_Mech=@n14
    Enable_Fill_Z_Weapon
    Fill_Z_Weapon_opacity
    Fill_Z_Weapon_Color
   ) {materialFunction=&/MF_TopsSideProjection_Base}.Combined) {materialFunction=&BreakOutFloat4Components}
  n33 = OneMinus(None=Multiply(A=@n32.R) {constB=0.25})
  n34 = Constant3Vector() {constant=<rgba 0 0 1 0>}
  n35 = VectorParameter() {group=Wrap_Subwoofer, parameterName=SubW_Grad_direction_Vehicle, sortPriority=500, defaultValue=<rgba 0 1 0 1>}
  n36 = LinearInterpolate(A=@n35.RGB, B=@n34, Alpha=@n11)
  n37 = MaterialFunctionCall!(
   Bounds=@n4
   ForwardAxis=LinearInterpolate(A=LinearInterpolate(A=@n34, B=@n36, Alpha=@n12), B=@n34, Alpha=@n14)
   Offset=Constant()
   FallOff=Constant() {r=1}
  ) {materialFunction=&/MF_Wrap_Gradent_Simple}
  n38 = OneMinus(None=@n37.Raw_Data)
  n39 = LinearInterpolate(A=LinearInterpolate(A=@n38, B=LinearInterpolate(A=@n37.Raw_Data, B=@n38, Alpha=@n11), Alpha=@n12), B=@n38, Alpha=@n14)
  n40 = LinearInterpolate(A=@n39, B=OneMinus(None=@n39), Alpha=ScalarParameter() {group=Wrap_Subwoofer, parameterName=SubW_Grad_invert, sortPriority=500})
  n41 = ComponentMask(None=@n40) {r=true}
  n42 = ScalarParameter() {defaultValue=0.5, group=Wrap_Subwoofer, parameterName=SubW_Stock_Contrast}
  n43 = LinearInterpolate(
   A=VectorParameter() {group=Wrap_Subwoofer, parameterName=SubW_Color_Backend, defaultValue=<rgba 0 0 0 1>}.RGB
   B=VectorParameter() {group=Wrap_Subwoofer, parameterName=SubW_Color_Background, defaultValue=<rgba 0.0307 0.0033 0.0545 1>}.RGB
   Alpha=Saturate(None=Power(Base=@n41, Exp=@n42))
  )
  n44 = MaterialFunctionCall(
   Hue Shift Percentage=Multiply(A=Time(), B=ScalarParameter() {group=Wrap_Subwoofer, parameterName=SubW_ColorCurve_HueShift_Speed})
   Texture=@n43
  ) {materialFunction=&HueShift}
  n45 = CollectionParameter() {collection=&FortniteMaterialParameters, parameterName=Emote_AudioLow}
  n46 = CollectionParameter() {collection=&FortniteMaterialParameters, parameterName=Emote_AudioHigh}
  n47 = Add(A=@n45, B=@n46)
  n48 = ScalarParameter() {group=Wrap_Subwoofer, parameterName=SubW_Emote_Mask}
  n49 = CollectionParameter() {collection=&FortniteMaterialParameters, parameterName=EventAudio_Amplitude}
  n50 = ScalarParameter() {group=Wrap_Subwoofer, parameterName=SubW_Event_Mask}
  n51 = Saturate(None=Max(A=Multiply(A=@n47, B=@n48) {constB=0.5}, B=Multiply(A=@n49, B=@n50) {constB=0.5}))
  n52 = Ceil(None=@n51)
  n53 = OneMinus(None=@n52)
  n54 = ScalarParameter() {defaultValue=22, group=Wrap_Subwoofer, parameterName="SubW_ColorCurve  Index"}
  n55 = OneMinus(None=@n40)
  n56 = CurveAtlasRowParameter(CurveTime=@n55) {
   atlas=/Game/Weapons/WeaponSkins/Wraps/Textures/CruveAtlas_Wraps_01
   curve=/Game/Weapons/WeaponSkins/Wraps/Textures/Subwoofer/C_SubwooferB
   defaultValue=22
   group=Wrap_Subwoofer
   parameterName=SubW_ColorCurve
  }
  n57 = ScalarParameter() {defaultValue=0.5, group=Wrap_Subwoofer, parameterName=SubW_ColorCruve_Default_Brightness}
  n58 = ScalarParameter() {defaultValue=1, group=Wrap_Subwoofer, parameterName=SubW_ColorCruve_Reactive_Brightness}
  n59 = Subtract(A=Time(), B=ScalarParameter() {defaultValue=-9999999, group=00-Reactive-GTFB, parameterName=00-GameTimeFromBP, sortPriority=100})
  n60 = ScalarParameter() {defaultValue=1, group=Wrap_Subwoofer, parameterName=SubW_Time_From_BP_Duration, sortPriority=100}
  n61 = OneMinus(None=Saturate(None=Divide(A=@n59, B=@n60)))
  n62 = LinearInterpolate(A=@n61, Alpha=ScalarParameter() {group=Wrap_Subwoofer, parameterName=SubW_reactive_disable, sortPriority=-1}) {constB=0}
  n63 = Multiply(A=OneMinus(None=@n52), B=@n62)
  n64 = Max(A=@n52, B=@n63)
  n65 = LinearInterpolate(
   A=ScalarParameter() {defaultValue=1, group=Wrap_Subwoofer, parameterName=SubW_Mask_G_Weapon, sortPriority=54}
   B=LinearInterpolate(
    A=ScalarParameter() {defaultValue=1, group=Wrap_Subwoofer, parameterName=SubW_Mask_G_Vehicle, sortPriority=54}
    B=ScalarParameter() {defaultValue=1, group=Wrap_Subwoofer, parameterName=SubW_Mask_G_VehicleSD, sortPriority=54}
   )
   Alpha=@n12
  )
  n66 = ScalarParameter() {defaultValue=1, group=Wrap_Subwoofer, parameterName=SubW_Mask_G_Mech, sortPriority=54}
  MaterialExpressionAppendVector_0 = AppendVector(
   A=Frac(None=ComponentMask(None=@n19) {r=true})
   B=Multiply(A=ComponentMask(None=@n19) {g=true}, B=LinearInterpolate(A=@n65, B=@n66, Alpha=@n14))
  )
  n67 = CollectionParameter() {collection=&FortniteMaterialParameters, parameterName=EventAudio_Amplitude}
  n68 = ScalarParameter() {group=Wrap_Subwoofer, parameterName=SubW_Sound_ampltude_effect, sortPriority=600}
  n69 = MaterialFunctionCall(Color=Multiply(A=@n5, B=Constant3Vector() {constant=<rgba 1 1 -1 0>})) {materialFunction=&SplitComponents}
  n70 = AppendVector(A=@n69.G, B=@n69.G)
  n71 = LinearInterpolate(A=AppendVector(A=@n69.R, B=@n69.R), B=AppendVector(A=@n69.R, B=@n69.R), Alpha=@n11)
  n72 = OneMinus(None=@n69.B)
  n73 = TextureSample(UVs=LinearInterpolate(A=LinearInterpolate(A=@n70, B=@n71, Alpha=@n12), B=AppendVector(A=@n72, B=@n69.R), Alpha=@n14)) {samplerType=SAMPLERTYPE_Masks, texture=/Game/Weapons/WeaponSkins/Wraps/Textures/Subwoofer/T_Wrap_Subwoofer_MB}
  n74 = CollectionParameter() {collection=&FortniteMaterialParameters, parameterName=EventAudio_Spectrum}
  n75 = ScalarParameter() {group=Wrap_Subwoofer, parameterName=SubW_Sound_Specturm strength, sortPriority=600}
  n76 = MaterialFunctionCall(Float4=Power(
    Base=Multiply(A=@n74, B=@n75) {constB=0.65}
    Exp=ScalarParameter() {defaultValue=1, group=Wrap_Subwoofer, parameterName=SubW_Sound_Specturm Power, sortPriority=600}
   )) {materialFunction=&BreakOutFloat4Components}
  n77 = Add(A=Add(A=Multiply(A=@n73.A, B=@n76.R), B=Multiply(A=@n73.B, B=@n76.G)), B=Add(A=Multiply(A=@n73.G, B=@n76.B), B=Multiply(A=@n73.R, B=@n76.A)))
  n78 = Add(
   A=Power(
    Base=Multiply(A=@n67, B=@n68) {constB=0.5}
    Exp=ScalarParameter() {defaultValue=1, group=Wrap_Subwoofer, parameterName=SubW_Sound_ampltude_Power, sortPriority=600}
   )
   B=@n77
  )
  MaterialExpressionMax_2 = Max(A=@n78, B=@n62) {constB=0}
  MaterialExpressionReroute_26 = Reroute(None=Max(A=@n78, B=Multiply(A=@n62) {constB=0.8}) {constB=0})
  n79 = ComponentMask(None=@n19) {r=true}
  MaterialExpressionLinearInterpolate_0 = LinearInterpolate(A=@MaterialExpressionMax_2, B=@MaterialExpressionReroute_26, Alpha=@n79)
  MaterialExpressionLinearInterpolate_7 = LinearInterpolate(A=@MaterialExpressionMax_2, B=@MaterialExpressionReroute_26, Alpha=@n79)
  MaterialExpressionLinearInterpolate_8 = LinearInterpolate(A=@MaterialExpressionMax_2, B=@MaterialExpressionReroute_26, Alpha=@n79)
  MaterialExpressionScalarParameter_16 = ScalarParameter() {defaultValue=3.5, group=Wrap_Subwoofer, parameterName=SubW_BandsWidth}
  MaterialExpressionCustom_0 = Custom!(
   uv=@MaterialExpressionAppendVector_0
   waveIntensity0=@MaterialExpressionMax_2
   waveIntensity1=@MaterialExpressionLinearInterpolate_0
   waveIntensity2=@MaterialExpressionLinearInterpolate_7
   waveIntensity3=@MaterialExpressionLinearInterpolate_8
   waveIntensity4=@MaterialExpressionReroute_26
   waveWidth=@MaterialExpressionScalarParameter_16
  ) {
   code="float waveA = -pow((waveWidth * uv.x - 0), 2) + waveIntensity0;\r\nfloat waveB = -pow((waveWidth * uv.x - waveWidth * 0.25), 2) + waveIntensity1;\r\nfloat waveC = -pow((waveWidth * uv.x - waveWidth * 0.5), 2) + waveIntensity2;\r\nfloat waveD = -pow((waveWidth * uv.x - waveWidth * 0.75), 2) + waveIntensity3;\r\nfloat waveE = -pow((waveWidth * uv.x - waveWidth), 2) + waveIntensity4;\r\nreturn max(waveA,max(waveB,max(waveC,max(waveD,waveE))));"
   description=CreateWave
   outputType=CMOT_Float1
   inputs [
    {inputName=uv, input {expression=&self:MaterialExpressionAppendVector_0}}
    {inputName=waveIntensity0, input {expression=&self:MaterialExpressionMax_2}}
    {
     inputName=waveIntensity1
     input {expression=&self:MaterialExpressionLinearInterpolate_0}
    }
    {
     inputName=waveIntensity2
     input {expression=&self:MaterialExpressionLinearInterpolate_7}
    }
    {
     inputName=waveIntensity3
     input {expression=&self:MaterialExpressionLinearInterpolate_8}
    }
    {
     inputName=waveIntensity4
     input {expression=&self:MaterialExpressionReroute_26}
    }
    {
     inputName=waveWidth
     input {expression=&self:MaterialExpressionScalarParameter_16}
    }
   ]
  }
  MaterialExpressionLinearInterpolate_18 = LinearInterpolate(
   A=ScalarParameter() {defaultValue=1, group=Wrap_Subwoofer, parameterName=SubW_VerticalOffset}
   B=LinearInterpolate(
    A=ScalarParameter() {group=Wrap_Subwoofer, parameterName=SubW_VerticalOffset_Vehicle}
    B=ScalarParameter() {group=Wrap_Subwoofer, parameterName=SubW_VerticalOffset_VehicleSD}
    Alpha=@n11
   )
   Alpha=@n12
  )
  n80 = Custom!(uv=@MaterialExpressionAppendVector_0, waveFunction=@MaterialExpressionCustom_0, verticalOffset=@MaterialExpressionLinearInterpolate_18) {
   code="float result = uv.y- verticalOffset + waveFunction;    \r\nreturn (0.1, result);"
   description=DrawWave
   outputType=CMOT_Float1
   inputs [
    {inputName=uv, input {expression=&self:MaterialExpressionAppendVector_0}}
    {inputName=waveFunction, input {expression=&self:MaterialExpressionCustom_0}}
    {
     inputName=verticalOffset
     input {expression=&self:MaterialExpressionLinearInterpolate_18}
    }
   ]
  }
  n81 = Saturate(None=@n80)
  n82 = ScalarParameter() {defaultValue=1, group=Wrap_Subwoofer, parameterName=SubW_Height_offset, sortPriority=49}
  n83 = Multiply(A=Max(A=@n51, B=@n63))
  n84 = Multiply(A=@n5, B=Constant3Vector() {constant=<rgba 1 1 -1 0>})
  n85 = VectorParameter() {group=Wrap_Subwoofer, parameterName=SubW_Sine_Speed_Weapon, defaultValue=<rgba 0 0 0 1>}
  n86 = LinearInterpolate(
   A=VectorParameter() {group=Wrap_Subwoofer, parameterName=SubW_Sine_Speed_vehicle, defaultValue=<rgba 0 0 0 1>}.RGB
   B=VectorParameter() {group=Wrap_Subwoofer, parameterName=SubW_Sine_Speed_vehicle_SP, defaultValue=<rgba 0 0 0 1>}.RGB
   Alpha=@n11
  )
  n87 = Multiply(
   A=Time()
   B=LinearInterpolate(
    A=LinearInterpolate(A=@n85.RGB, B=@n86, Alpha=@n12)
    B=VectorParameter() {group=Wrap_Subwoofer, parameterName=SubW_Sine_Speed_Mech, defaultValue=<rgba 0 0 0 1>}.RGB
    Alpha=@n14
   )
  )
  n88 = LinearInterpolate(
   A=VectorParameter() {group=Wrap_Subwoofer, parameterName=SubW_Sine_offset_Weapon, defaultValue=<rgba 0 0 1 1>}.RGB
   B=LinearInterpolate(
    A=VectorParameter() {group=Wrap_Subwoofer, parameterName=SubW_Sine_offset_vehicle, defaultValue=<rgba 0 0 0 1>}.RGB
    B=VectorParameter() {group=Wrap_Subwoofer, parameterName=SubW_Sine_offset_vehicle_SP, defaultValue=<rgba 0 0 0 1>}.RGB
    Alpha=@n11
   )
   Alpha=@n12
  )
  n89 = VectorParameter() {group=Wrap_Subwoofer, parameterName=SubW_Sine_offset_Mech, defaultValue=<rgba 0 0 0 1>}
  n90 = MaterialFunctionCall(Color=Add(A=Add(A=@n84, B=@n87), B=LinearInterpolate(A=@n88, B=@n89.RGB, Alpha=@n14))) {materialFunction=&SplitComponents}
  n91 = MaterialFunctionCall(UVs=AppendVector(A=@n90.R, B=@n90.R), "Rotation Angle (0-1)"=@n8) {materialFunction=&CustomRotator}
  n92 = AppendVector(A=@n90.R, B=@n90.R)
  n93 = LinearInterpolate(
   A=LinearInterpolate(A=AppendVector(A=@n90.G, B=@n90.G), B=LinearInterpolate(A=@n91.Rotated Values, B=@n92, Alpha=@n11), Alpha=@n12)
   B=AppendVector(A=@n90.R, B=@n90.R)
   Alpha=@n14
  )
  n94 = ComponentMask(None=@n93) {r=true}
  n95 = ScalarParameter() {defaultValue=1, group=Wrap_Subwoofer, parameterName=SubW_Sine_tile}
  n96 = ComponentMask(None=@n93) {g=true}
  n97 = Time()
  n98 = ScalarParameter() {group=Wrap_Subwoofer, parameterName=SubW_Sine_Z_Speed}
  n99 = Multiply(A=Sine(None=Multiply(A=@n97, B=@n98)), B=ScalarParameter() {group=Wrap_Subwoofer, parameterName=SubW_Sine_Z_Amount})
  n100 = TextureSample(
   UVs=AppendVector(A=Multiply(A=@n94, B=@n95), B=Add(A=@n96, B=@n99))
   Tex=TextureObject() {samplerType=SAMPLERTYPE_LinearGrayscale, texture=/Game/Weapons/WeaponSkins/Wraps/Textures/Subwoofer/T_Wrap_Subwoofer_grad}
  )
  n101 = MaterialFunctionCall(Color=@n5) {materialFunction=&SplitComponents}
  n102 = MaterialFunctionCall(UVs=AppendVector(A=@n101.R, B=@n101.R), "Rotation Angle (0-1)"=@n8) {materialFunction=&CustomRotator}
  n103 = AppendVector(A=@n101.R, B=@n101.R)
  n104 = LinearInterpolate(
   A=LinearInterpolate(A=AppendVector(A=@n101.G, B=@n101.G), B=LinearInterpolate(A=@n102.Rotated Values, B=@n103, Alpha=@n11), Alpha=@n12)
   B=AppendVector(A=@n101.R, B=@n101.R)
   Alpha=@n14
  )
  n105 = Multiply(
   A=LinearInterpolate(
    A=ChannelMaskParameter(None=@n100.RGBA) {group=Wrap_Subwoofer, parameterName=SubW_Sine_Mask, sortPriority=15}.R
    B=Multiply(A=@n81, B=@n82) {constB=5}
    Alpha=Saturate(None=@n83)
   )
   B=SphereMask(A=ComponentMask(None=@n104) {g=true}, B=Constant() {r=0.5}, Radius=Constant() {r=0.5}, Hardness=Constant() {r=1})
  )
  n106 = Multiply(A=@n32.A, B=@n105)
  n107 = Multiply(A=@n106, B=Add(A=@n106, B=@n105))
  n108 = ScalarParameter() {group=Wrap_Subwoofer, parameterName=SubW_Height_Contrast, sortPriority=49}
  n109 = LinearInterpolate(
   A=Multiply(A=@n33, B=LinearInterpolate(A=@n43, B=@n44.Result, Alpha=@n53))
   B=Multiply(
    A=MaterialFunctionCall!(Index=@n54, CCinput=@n56, CCinput_Alpha, UV=@n55) {materialFunction=&/MF_Global_CC_Tex}.3V
    B=LinearInterpolate(A=@n57, B=@n58, Alpha=@n64)
   )
   Alpha=Saturate(None=MaterialFunctionCall(In=@n107, Contrast=@n108) {materialFunction=&CheapContrast}.Result)
  )
  n110 = Multiply(A=ScalarParameter() {defaultValue=1, group=Wrap_Subwoofer, parameterName=SubW_Edge_Brightness}, B=@n109)
  n111 = FunctionInput() {inputName=Material, inputType=FunctionInput_MaterialAttributes}
  n112 = GetMaterialAttributes(None=@n111) {attributeGetTypes [#1, #2, #3]}
  n113 = SetMaterialAttributes!(
   MaterialAttributes=@n112.MaterialAttributes
   BaseColor=VectorParameter() {group=Wrap_Subwoofer, parameterName=SUBW_BaseColor, sortPriority=-5, defaultValue=<rgba 0 0 0 1>}.RGB
   EmissiveColor=Add(A=Multiply(A=@n25.Result, B=@n110), B=@n109)
   Roughness=ScalarParameter() {defaultValue=0.25, group=Wrap_Subwoofer, parameterName=SubW_Roughness, sortPriority=-5}
   Specular=Multiply(A=ScalarParameter() {defaultValue=0.5, group=Wrap_Subwoofer, parameterName=SubW_Specular, sortPriority=-5}, B=@n33)
  ) {attributeSetTypes [
    #1
    #4
    #2
    #5
   ]}
  n114 = FunctionInput() {bUsePreviewValueAsDefault=true, inputName="Customization Mask (_CM B)", inputType=FunctionInput_Scalar, sortPriority=24}
  n115 = ScalarParameter() {group=Wrap_Subwoofer, parameterName=SubW_Remove_Customization Mask, sortPriority=-10}
  n116 = BlendMaterialAttributes(A=@n112.MaterialAttributes, B=@n113, Alpha=Saturate(None=Add(A=@n114, B=@n115)))
  FunctionOutput(None=StaticSwitchParameter(True=@n116, False=@n111) {group=Customization_Wrap, parameterName=Wrap_Subwoofer})
  n117 = ScalarParameter() {defaultValue=2, group=Wrap_Subwoofer, parameterName=SubW_SpeedBass}
  n118 = ScalarParameter() {defaultValue=2, group=Wrap_Subwoofer, parameterName=SubW_SpeedBass_Audio}
  n119 = Time()
  MaterialExpressionFrac_3 = Frac(None=Multiply(A=LinearInterpolate(A=@n117, B=@n118, Alpha=@n52), B=@n119))
  n120 = Custom!(InputFunction=@MaterialExpressionFrac_3) {
   code="float result = (2 * InputFunction) -2;\r\nresult  = result  * result  * result ;\r\nreturn -1.181 * InputFunction * result;"
   description=FastAttackSlowDecay
   outputType=CMOT_Float1
   inputs [{inputName=InputFunction, input {expression=&self:MaterialExpressionFrac_3}}]
  }
  n121 = MaterialFunctionCall(Float4=@n74) {materialFunction=&BreakOutFloat4Components}
  n122 = Add(A=Multiply(A=@n121.R) {constB=0.75}, B=Multiply(A=@n121.B) {constB=0.75})
  Max(
   A=Multiply(A=@n120, B=ScalarParameter() {defaultValue=0.4, group=Wrap_Subwoofer, parameterName=SubW_MaxIntensityBass})
   B=Max(A=@n45, B=Power(Base=@n122))
  ) {constB=0}
  n123 = ScalarParameter() {defaultValue=1.2, group=Wrap_Subwoofer, parameterName=SubW_SpeedTreble}
  n124 = ScalarParameter() {defaultValue=1.2, group=Wrap_Subwoofer, parameterName=SubW_SpeedTreble_audio}
  MaterialExpressionFrac_4 = Frac(None=Multiply(A=@n119, B=LinearInterpolate(A=@n123, B=@n124, Alpha=@n52)) {desc="frac(time * speed)"})
  n125 = Custom(InputFunction=@MaterialExpressionFrac_4) {
   code="float result = (2 * InputFunction) -2;\r\nresult  = result  * result  * result ;\r\nreturn -1.181 * InputFunction * result;"
   description=FastAttackSlowDecay
   outputType=CMOT_Float1
   inputs [{inputName=InputFunction, input {expression=&self:MaterialExpressionFrac_4}}]
  }
  n126 = Multiply(A=@n49) {constB=0.65}
  Max(
   A=Multiply(A=ScalarParameter() {defaultValue=0.5, group=Wrap_Subwoofer, parameterName=SubW_MaxIntensityTreble}, B=@n125)
   B=Max(A=@n46, B=Power(Base=@n126) {constExponent=3})
  )
  Multiply(A=@n61)
  Multiply(B=@n77)
  StaticSwitchParameter(True=@n116, False=Constant()) {parameterName=446asdfa}
 ]}
