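# Weapon wrap material function for the TigerRoot skin, blending animated base and flat accent materials over the root material by mask, behind a static switch.
# SafeZone virtual-plane UVs become twisting radial coordinates for warped curve-atlas color and layered stripe swirls, with Fresnel darkening.
# Music amplitude and a timed reactive fade form an AudioSpectrum signal for UV bounce and color shift.
# Explicit derivatives control mip selection on the animated texture samples.
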MaterialFunction /Game/Weapons/WeaponSkins/Wraps/MaterialFunctions/MF_Wrap_TigerR
graph {nodes [
  n1 = ScalarParameter() {defaultValue=0.5, group=Wrap_TigerRoot_Audio, parameterName=TigR_Audio_Bounce_Strenght}
  n2 = NamedRerouteUsage()
  n3 = NamedRerouteUsage()
  n4 = NamedRerouteUsage()
  n5 = Multiply(
   A=LinearInterpolate(
    A=ScalarParameter() {defaultValue=1, group=Wrap_TigerRoot_Audio, parameterName=TigR_Audio_Bounce_A}
    B=ScalarParameter() {defaultValue=0.58, group=Wrap_TigerRoot_Audio, parameterName=TigR_Audio_Bounce_B}
    Alpha=Multiply(A=@n1, B=@n2)
   )
   B=Add(A=MaterialFunctionCall!(
     Camera _pos
     Axis
     Scale
     Scale_Multplyer
     Object-SafeZone-Virtual-Pivot=@n3
     Object-SafeZone-Virtual-CH-Switch=@n4
     SinglePlane Normal Axis
     SinglePlane X Axis
     PlaneCenter_distortion
    ) {materialFunction=&/MF_SafeZone_VirtualPlane}.Uvs) {constB=-0.5}
  ) {constA=1}
  Add(A=@n5) {constB=0.5}
  n6 = TextureSample() {samplerType=SAMPLERTYPE_Masks, texture=/Game/Weapons/FORT_Rifles/Textures/T_Scar_CM}
  n7 = FunctionInput(Preview=@n6.R) {bUsePreviewValueAsDefault=true, inputName=Mask Primary, inputType=FunctionInput_Scalar, sortPriority=24, previewValue=<1 0 0 1>}
  n8 = FunctionInput(Preview=@n6.B) {bUsePreviewValueAsDefault=true, inputName=Mask Secondary, inputType=FunctionInput_Scalar, sortPriority=24, previewValue=<1 0 0 1>}
  n9 = BlendMaterialAttributes(A=NamedRerouteUsage(), B=NamedRerouteUsage(), Alpha=Saturate(None=Add(A=@n7, B=@n8))) {vertexAttributeBlendType=UseA}
  n10 = NamedRerouteUsage()
  n11 = NamedRerouteDeclaration(None=@n8) {name=Msk_Accent, nodeColor=<rgba 0 1 0.7656 1>}
  n12 = FunctionInput(Preview=SetMaterialAttributes!(
    MaterialAttributes
    Normal=TextureSample() {samplerType=SAMPLERTYPE_Normal, texture=/Game/Weapons/FORT_Rifles/T_SCAR_N}.RGB
    BaseColor=Constant3Vector() {constant=<rgba 1 0 0 0>}
   ) {attributeSetTypes [#1, #2]}) {inputName=Material_Root, inputType=FunctionInput_MaterialAttributes}
  FunctionOutput(None=StaticSwitchParameter(
    True=BlendMaterialAttributes(A=@n9, B=@n10, Alpha=@n11) {vertexAttributeBlendType=UseA}
    False=FunctionInput(Preview=@n12) {bUsePreviewValueAsDefault=true, inputName=Material_1, inputType=FunctionInput_MaterialAttributes}
   ) {group=Customization_Wrap, parameterName=Wrap_TigerRoot})
  n13 = MaterialFunctionCall() {materialFunction=&ObjectScale}
  n14 = NamedRerouteUsage()
  n15 = LinearInterpolate(
   A=TransformPosition(None=ComponentMask(None=@n14) {b=true, g=true, r=true}) {transformType=TRANSFORMPOSSOURCE_World}
   B=ComponentMask(None=@n14) {b=true, g=true, r=true}
   Alpha=NamedRerouteUsage()
  )
  n16 = ComponentMask(None=@n14) {a=true}
  n17 = PreSkinnedLocalBounds()
  n18 = MaterialFunctionCall(Float3=Subtract(A=@n17.Max, B=@n17.Max)) {materialFunction=&BreakOutFloat3Components}
  n19 = Max(A=Max(A=@n18.R, B=@n18.R), B=@n18.B)
  Multiply(
   A=Multiply(
    A=Divide(B=@n13.Scale Z)
    B=MaterialFunctionCall!(Center Point World Position=@n15) {materialFunction=/Engine/Functions/Engine_MaterialFunctions02/UVs/WorldSpaceAlignedScreenCoordinates}.Offset and Scaled UVS
   )
   B=Divide(A=Constant() {r=-2500}, B=Multiply(A=@n16, B=@n19))
  )
  n20 = Fresnel(ExponentIn=ScalarParameter() {defaultValue=5, group=Wrap_TigerRoot_Edge, parameterName=TigR_Fresnel_Exp}) {baseReflectFraction=0}
  Multiply(A=@n20)
  Multiply(A=MaterialFunctionCall(Value_Morning=Constant() {r=0.5}, Value_Day=Constant() {r=1}, Value_Evening=Constant() {r=0.5}, Value_Night=Constant() {r=0.25}) {materialFunction=&Scalar_Time_of_Day}.Result)
  n21 = NamedRerouteUsage()
  n22 = Abs(None=Multiply(A=Add(A=@n21) {constB=-0.5}) {constB=5})
  n23 = Power(Base=OneMinus(None=Saturate(None=@n22)))
  n24 = ScalarParameter() {defaultValue=0.3, group=Wrap_TigerRoot_Edge, parameterName=TigR_Fresnel_DarkValues}
  n25 = Add(A=Power(Base=OneMinus(None=@n20)), B=@n24) {constB=0.5}
  Multiply(A=Saturate(None=Add(A=@n23, B=@n24)), B=@n25)
  n26 = NamedRerouteUsage()
  Multiply(A=@n26) {constB=0}
  n27 = VectorParameter() {group=Wrap_TigerRoot_Accent, parameterName=TigR_Accent_BaseColor, defaultValue=<rgba 0.0296 0.0037 0.1221 1>}
  n28 = SetMaterialAttributes!(
   MaterialAttributes=NamedRerouteUsage()
   BaseColor=@n27.RGB
   Roughness=ScalarParameter() {defaultValue=0.6, group=Wrap_TigerRoot_Accent, parameterName=TigR_Accent_Roughness}
   Specular=ScalarParameter() {defaultValue=0.5, group=Wrap_TigerRoot_Accent, parameterName=TigR_Accent_Spec}
   Metallic=ScalarParameter() {group=Wrap_TigerRoot_Accent, parameterName=TigR_Accent_Metalic}
   Normal
   EmissiveColor
  ) {attributeSetTypes [
    #2
    #3
    #4
    #5
    #1
    #6
   ]}
  NamedRerouteDeclaration(None=@n28) {name=Mat_Accent, nodeColor=<rgba 0 0.3906 1 1>}
  n29 = FunctionInput(Preview=@n6.G) {bUsePreviewValueAsDefault=true, inputName=Mask Crevise, inputType=FunctionInput_Scalar, sortPriority=24, previewValue=<1 0 0 1>}
  NamedRerouteDeclaration(None=@n29) {name=Msk_Curvature, nodeColor=<rgba 0 1 0.7422 1>}
  n30 = GetMaterialAttributes(None=@n12) {attributeGetTypes [
    #2
    #3
    #1
    #5
   ]}
  NamedRerouteDeclaration(None=@n30.MaterialAttributes) {name=Root_Material, nodeColor=<rgba 0.7812 1 0 1>}
  n31 = FunctionInput(Preview=Constant4Vector() {constant=<rgba 0 0 5 -1>}.RGBA) {bUsePreviewValueAsDefault=true, inputName=Object-SafeZone-Virtual-Pivot, inputType=FunctionInput_Vector4, sortPriority=105}
  NamedRerouteDeclaration(None=@n31) {name=SafeZone VP, nodeColor=<rgba 0.5 1 0 1>}
  n32 = FunctionInput(Preview=Constant()) {bUsePreviewValueAsDefault=true, inputName=Object-SafeZone-VP-Character-PreView-Switch, inputType=FunctionInput_Scalar, sortPriority=105}
  NamedRerouteDeclaration(None=@n32) {name=SafeZone Character-Switch, nodeColor=<rgba 1 0.9375 0 1>}
  n33 = MaterialFunctionCall(Vector or UVs=NamedRerouteUsage()) {materialFunction=&VectorToRadialValue}
  n34 = Power(
   Base=ComponentMask(None=@n33.Radial Coordinates) {g=true}
   Exp=ScalarParameter() {defaultValue=0.2, group=Wrap_TigerRoot_BackGround, parameterName=TigR_Rotation_Pow}
  )
  n35 = ScalarParameter() {defaultValue=5, group=Wrap_TigerRoot_BackGround, parameterName=TigR_Rotation_Amount}
  n36 = Time()
  n37 = ScalarParameter() {defaultValue=-0.1, group=Wrap_TigerRoot_BackGround, parameterName=TigR_Rotation_Speed}
  n38 = Add(A=@n33.Radial Coordinates, B=Add(A=Multiply(A=@n34, B=@n35), B=Multiply(A=@n36, B=@n37)))
  NamedRerouteDeclaration(None=@n38) {name=Uvs_Rotation, nodeColor=<rgba 1 0.3281 0 1>}
  n39 = NamedRerouteUsage()
  n40 = ComponentMask(None=VectorParameter() {group=Wrap_TigerRoot_Stripes, parameterName=TigR_StripsC_TileRate, defaultValue=<rgba 1 0.005 0 1>}.RGB) {g=true, r=true}
  n41 = NamedRerouteUsage()
  NamedRerouteDeclaration(None=TextureSample!(UVs=Multiply(A=@n39, B=@n40), Tex=NamedRerouteUsage(), "DDX(UVs)"=DDX(Value=@n41), "DDY(UVs)"=DDY(Value=@n41), Apply View MipBias) {mipValueMode=TMVM_Derivative, samplerType=SAMPLERTYPE_Masks}.G) {name=Strips_C, nodeColor=<rgba 0 1 0.6484 1>}
  n42 = Multiply(
   A=Add(A=NamedRerouteUsage()) {constB=-0.5}
   B=ScalarParameter() {defaultValue=8, group=Wrap_TigerRoot_Edge, parameterName=TigR_Edge_Thickness}
  ) {constB=30}
  n43 = Saturate(None=Abs(None=@n42))
  n44 = ScalarParameter() {defaultValue=2, group=Wrap_TigerRoot_Edge, parameterName=TigR_Edge_Thickness_Pow}
  n45 = Multiply(
   A=Saturate(None=Power(Base=@n43, Exp=@n44) {constExponent=10})
   B=ScalarParameter() {defaultValue=1, group=Wrap_TigerRoot_Edge, parameterName=TigR_Edge_Thickness_Opacity}
  )
  n46 = ScalarParameter() {defaultValue=0.3, group=Wrap_TigerRoot, parameterName=TigR_Rougness}
  n47 = NamedRerouteUsage()
  n48 = Saturate(None=Multiply(A=Add(A=@n47) {constB=-0.5}) {constB=2})
  n49 = SetMaterialAttributes(
   MaterialAttributes=NamedRerouteUsage()
   BaseColor=LinearInterpolate(
    A=VectorParameter() {group=Wrap_TigerRoot, parameterName=TigR_BaseColor_A, defaultValue=<rgba 0.002 0.002 0.0025 1>}.RGB
    B=VectorParameter() {group=Wrap_TigerRoot, parameterName=TigR_BaseColorr_B, defaultValue=<rgba 0.0583 0.0534 0.1 1>}.RGB
    Alpha=@n45
   )
   Roughness=Saturate(None=Add(A=@n46, B=@n48))
   Specular=ScalarParameter() {defaultValue=1, group=Wrap_TigerRoot, parameterName=TigR_Spec}
   Metallic=ScalarParameter() {group=Wrap_TigerRoot, parameterName=TigR_Metalic}
   EmissiveColor=LinearInterpolate(A=@n26, Alpha=@n45) {constB=0}
  ) {attributeSetTypes [
    #2
    #3
    #4
    #5
    #1
    #6
   ]}
  NamedRerouteDeclaration(None=@n49) {name=Mat_Base, nodeColor=<rgba 0 0.3906 1 1>}
  n50 = ScalarParameter() {defaultValue=129, group=Wrap_TigerRoot_BackGround, parameterName="TigR_Bck_CC  Index"}
  n51 = ScalarParameter() {defaultValue=0.1, group=Wrap_TigerRoot_BackGround, parameterName=TigR_Audio_ColorShift_Strength}
  n52 = NamedRerouteUsage()
  n53 = ComponentMask(None=@n38) {r=true}
  n54 = VectorParameter() {group=Wrap_TigerRoot_BackGround, parameterName=TigR_Bck_TileRate, defaultValue=<rgba 1 0.2 0 1>}
  n55 = VectorParameter() {group=Wrap_TigerRoot_BackGround, parameterName=TigR_Bck_Speed, defaultValue=<rgba 0.15 0.1 0 1>}
  n56 = Add(A=Multiply(A=@n38, B=ComponentMask(None=@n54.RGB) {g=true, r=true}), B=Multiply(A=ComponentMask(None=@n55.RGB) {g=true, r=true}, B=Time()))
  n57 = NamedRerouteUsage()
  n58 = NamedRerouteUsage()
  n59 = DDX(Value=@n58)
  n60 = DDY(Value=@n58)
  n61 = Multiply(
   A=Add(A=TextureSample(UVs=@n56, Tex=@n57, "DDX(UVs)"=@n59, "DDY(UVs)"=@n60) {mipValueMode=TMVM_Derivative, samplerType=SAMPLERTYPE_Masks}.A) {constB=-0.5}
   B=ScalarParameter() {defaultValue=1, group=Wrap_TigerRoot_BackGround, parameterName=TigR_Bck_Warp_Amount}
  )
  n62 = Frac(None=Add(A=Multiply(A=@n51, B=@n52), B=Add(A=@n53, B=@n61)))
  n63 = CurveAtlasRowParameter(CurveTime=@n62) {
   atlas=/Game/Weapons/WeaponSkins/Wraps/Textures/CruveAtlas_Wraps_01
   curve=/Game/Weapons/WeaponSkins/Wraps/Textures/0_ColorCurves/CC_TigR_ColorsA
   defaultValue=129
   group=Wrap_TigerRoot_BackGround
   parameterName=TigR_Bck_CC
  }
  n64 = NamedRerouteUsage()
  n65 = Constant2Vector() {g=0.23, r=0.3}
  n66 = VectorParameter() {group=Wrap_TigerRoot_Stripes, parameterName=TigR_StripsB_TileRate, defaultValue=<rgba 1 0.005 0 1>}
  n67 = VectorParameter() {group=Wrap_TigerRoot_Stripes, parameterName=TigR_StripsB_Speed, defaultValue=<rgba 0.133 0.1 0 1>}
  n68 = Add(
   A=Multiply(A=Add(A=@n64, B=@n65) {constB=0.3}, B=ComponentMask(None=@n66.RGB) {g=true, r=true})
   B=Multiply(A=ComponentMask(None=@n67.RGB) {g=true, r=true}, B=Time())
  )
  n69 = NamedRerouteUsage()
  n70 = NamedRerouteUsage()
  n71 = DDX(Value=@n70)
  n72 = DDY(Value=@n70)
  n73 = VectorParameter() {group=Wrap_TigerRoot_Stripes, parameterName=TigR_StripsA_TileRate, defaultValue=<rgba 1 0.0051 0 1>}
  n74 = VectorParameter() {group=Wrap_TigerRoot_Stripes, parameterName=TigR_StripsA_Speed, defaultValue=<rgba 0.15 0.1 0 1>}
  n75 = Add(
   A=Multiply(A=NamedRerouteUsage(), B=ComponentMask(None=@n73.RGB) {g=true, r=true})
   B=Multiply(A=ComponentMask(None=@n74.RGB) {g=true, r=true}, B=Time())
  )
  n76 = NamedRerouteUsage()
  n77 = NamedRerouteUsage()
  n78 = DDX(Value=@n77)
  n79 = DDY(Value=@n77)
  n80 = Add(
   A=LinearInterpolate(
    A=MaterialFunctionCall!(Index=@n50, CCinput=@n63, CCinput_Alpha, UV=@n62) {materialFunction=&/MF_Global_CC_Tex}.3V
    B=VectorParameter() {group=Wrap_TigerRoot_Stripes, parameterName=TigR_StripsB_Color, defaultValue=<rgba 1 0 0.7854 1>}.RGB
    Alpha=TextureSample(UVs=@n68, Tex=@n69, "DDX(UVs)"=@n71, "DDY(UVs)"=@n72) {mipValueMode=TMVM_Derivative, samplerType=SAMPLERTYPE_Masks}.R
   )
   B=Multiply(
    A=VectorParameter() {group=Wrap_TigerRoot_Stripes, parameterName=TigR_StripsA_Color, defaultValue=<rgba 0.2 0.2 0.2 1>}.RGB
    B=TextureSample(UVs=@n75, Tex=@n76, "DDX(UVs)"=@n78, "DDY(UVs)"=@n79) {mipValueMode=TMVM_Derivative, samplerType=SAMPLERTYPE_Masks}.R
   )
  )
  n81 = VectorParameter() {group=Wrap_TigerRoot_Stripes, parameterName=TigR_StripsC_Color, defaultValue=<rgba 10 10 10 0.5>}
  n82 = Multiply(A=NamedRerouteUsage(), B=NamedRerouteUsage()) {constB=0}
  NamedRerouteDeclaration(None=Multiply(A=Saturate(None=@n25), B=LinearInterpolate(A=@n80, B=@n81.RGB, Alpha=@n82))) {name=Swirls, nodeColor=<rgba 1 0 0.75 1>}
  NamedRerouteDeclaration(None=@n5) {name=UVs_VP, nodeColor=<rgba 0 1 0.5078 1>}
  n83 = Add(A=NamedRerouteUsage()) {constB=-0.5}
  n84 = ScalarParameter() {defaultValue=-0.01, group=Wrap_TigerRoot, parameterName=TigR_Uv_Warp_Curvature}
  NamedRerouteDeclaration(None=Add(A=@n5, B=Multiply(A=@n83, B=@n84))) {name=UVs_VP_Warped, nodeColor=<rgba 0 1 0.5078 1>}
  NamedRerouteDeclaration(None=FunctionInput(Preview=Constant4Vector().RGBA) {bUsePreviewValueAsDefault=true, inputName=IsState, inputType=FunctionInput_Vector4, sortPriority=4}) {name=IsState, nodeColor=<rgba 0.1497 1 0.1633 1>}
  n85 = CollectionParameter() {collection=&FortniteMaterialParameters, parameterName=EventAudio_Amplitude}
  n86 = CollectionParameter() {collection=&FortniteMaterialParameters, parameterName=Emote_AudioLow}
  n87 = Saturate(None=MaterialFunctionCall(
    Input=Max(A=@n85, B=@n86)
    Input Low=ScalarParameter() {defaultValue=0.25, group=Wrap_TigerRoot_Audio, parameterName=TigR_Sectrum_Low_Remap, sortPriority=150}
    Input High=ScalarParameter() {defaultValue=1, group=Wrap_TigerRoot_Audio, parameterName=TigR_Sectrum_High_Remap, sortPriority=150}
    Target Low=Constant()
    Target High=Constant() {r=1}
   ) {materialFunction=&RemapValueRange}.Result)
  n88 = CollectionParameter() {collection=&FortniteMaterialParameters, parameterName=Emote_AudioHigh}
  n89 = ScalarParameter() {defaultValue=1, group=Wrap_TigerRoot_Reactive, parameterName=TigR_reactive_Start}
  n90 = ScalarParameter() {group=Wrap_TigerRoot_Reactive, parameterName=TigR_reactive_End}
  n91 = Subtract(
   A=Time()
   B=Add(
    A=ScalarParameter() {defaultValue=-9999999, group=Wrap_Reactive, parameterName=TimeFrom-BP----------------, sortPriority=600}
    B=ScalarParameter() {group=Wrap_TigerRoot_Reactive, parameterName=TigR_reactive_Delay}
   )
  )
  n92 = ScalarParameter() {defaultValue=0.5, group=Wrap_TigerRoot_Reactive, parameterName=TigR_reactive_Duration}
  n93 = Power(
   Base=Saturate(None=Divide(A=@n91, B=@n92))
   Exp=ScalarParameter() {defaultValue=1, group=Wrap_TigerRoot_Reactive, parameterName=TigR_reactive_Falloff}
  )
  NamedRerouteDeclaration(None=Max(A=Max(A=@n87, B=@n88), B=LinearInterpolate(A=@n89, B=@n90, Alpha=@n93))) {name=AudioSpectrum, nodeColor=<rgba 1 0.8203 0 1>}
  NamedRerouteDeclaration(None=FunctionInput(Preview=Constant2Vector()) {bUsePreviewValueAsDefault=true, inputName=IsState_B, inputType=FunctionInput_Vector2, sortPriority=4}) {name=IsState_B, nodeColor=<rgba 0.1497 1 0.1633 1>}
  n94 = PreSkinnedLocalBounds()
  n95 = Subtract(A=VertexInterpolator(VS=PreSkinnedPosition()).PS, B=@n94.Min)
  n96 = MaterialFunctionCall(Color=Subtract(A=@n94.Max, B=@n94.Max)) {materialFunction=&SplitComponents}
  n97 = Max(A=Max(A=@n96.R, B=@n96.R), B=@n96.B)
  NamedRerouteDeclaration(None=FunctionInput(Preview=Divide(A=@n95, B=@n97)) {bUsePreviewValueAsDefault=true, inputName=Calcated_Bounds, sortPriority=11}) {name=Caclcated_Bounds, nodeColor=<rgba 0 0.9297 1 1>}
  NamedRerouteDeclaration(None=TextureObjectParameter() {group=Wrap_TigerRoot, parameterName=TigR__Tex_MskA, texture=&T_BlacksRGB}) {name=Tex_MskA, nodeColor=<rgba 1 0.8672 0 1>}
  n98 = PreSkinnedNormal()
  NamedRerouteDeclaration(None=FunctionInput(Preview=VertexInterpolator(VS=@n98).PS) {bUsePreviewValueAsDefault=true, inputName=Pre-Skinned Local Normal, sortPriority=10}) {name=PreSkin_Normal, nodeColor=<rgba 0.8281 1 0 1>}
  NamedRerouteDeclaration(None=@n30.Normal) {name=Root_Normal, nodeColor=<rgba 0.4062 0 1 1>}
  NamedRerouteUsage()
  n99 = Multiply(A=@n42) {constB=-1}
  Power(Base=Add(A=Saturate(None=@n99), B=Saturate(None=@n42)) {constB=0}) {constExponent=5}
  ScalarParameter() {defaultValue=-9999999, group=TopSideProjection_Emissive_Reactive, parameterName=Time_BP}
 ]}
