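# Material function for the TurboC wrap skin, replacing incoming material attributes with an emissive layered look when the Wrap_TurboCar static switch is on.
# A camera-facing virtual plane drives radial background glow, overlaid with curve-atlas-colored edges from the customization mask's green channel.
# A projected, clamped mask texture adds a second colored layer; both fake depth with stacked view-offset samples.
# Is_States selects weapon, vehicle, vehicle-side or mech tuning, and time of day scales emissive.
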MaterialFunction /Game/Weapons/WeaponSkins/Wraps/MaterialFunctions/MF_Wrap_TurboC
graph {nodes [
  n1 = VertexInterpolator(VS=PreSkinnedPosition())
  n2 = PreSkinnedLocalBounds()
  n3 = Add(
   A=@n2.Min
   B=FunctionInput(Preview=Constant3Vector() {constant=<rgba 0 5 0 0>}) {bUsePreviewValueAsDefault=true, inputName=Object Bounds Min Adjustment, sortPriority=50}
  )
  n4 = Add(A=Multiply(A=@n1.PS) {constB=0}, B=@n3)
  n5 = Constant3Vector() {constant=<rgba -5 -25 -5 0>}
  n6 = Subtract(
   A=Add(A=@n2.Max, B=FunctionInput(Preview=@n5) {bUsePreviewValueAsDefault=true, inputName=Object Bounds Max Adjustment, sortPriority=50})
   B=@n3
  )
  n7 = Multiply(A=@n6) {constA=0.5, constB=0.5}
  n8 = MaterialFunctionCall(Float4=FunctionInput() {inputName=Is_States, inputType=FunctionInput_Vector4, sortPriority=5}) {materialFunction=&BreakOutFloat4Components}
  n9 = MaterialFunctionCall(R=@n8.R, G=@n8.R, B=@n8.R) {materialFunction=&AppendMany}
  n10 = MaterialFunctionCall(Color=@n9.RGB) {materialFunction=&SplitComponents}
  n11 = LinearInterpolate(A=TransformPosition(None=Add(A=@n4, B=@n7)) {transformType=TRANSFORMPOSSOURCE_World}, B=ObjectPositionWS(), Alpha=@n10.B)
  n12 = Normalize(VectorInput=Subtract(A=@n11, B=CameraPositionWS()))
  n13 = CrossProduct(A=@n12, B=Constant3Vector() {constant=<rgba 0 0 1 1>})
  n14 = MaterialFunctionCall(Color=@n6) {materialFunction=&SplitComponents}
  n15 = Max(A=Max(A=@n14.R, B=@n14.R), B=@n14.B)
  n16 = MaterialFunctionCall() {materialFunction=/Engine/Functions/Engine_MaterialFunctions02/WorldPositionOffset/ObjectScale}
  n17 = LinearInterpolate(
   A=MaterialFunctionCall(Plane Normal Axis=@n12, Plane X Axis=Normalize(VectorInput=@n13), Plane Center=@n11, UV Scale=Multiply(A=@n15, B=@n16.Scale Y)) {materialFunction=&VirtualPlaneCoordinates}.UVs
   B=NamedRerouteUsage()
   Alpha=NamedRerouteUsage()
  )
  n18 = Constant() {r=0.5}
  n19 = ScalarParameter() {defaultValue=0.5, group=Wrap_TurboCar, parameterName=TurboCar_GlowBK_Radius_Vehicle}
  n20 = ScalarParameter() {defaultValue=0.5, group=Wrap_TurboCar, parameterName=TurboCar_GlowBK_Radius_VehicleSide}
  n21 = MaterialFunctionCall(Color=@n9.RGB) {materialFunction=&SplitComponents}
  n22 = LinearInterpolate(
   A=LinearInterpolate(
    A=ScalarParameter() {defaultValue=0.5, group=Wrap_TurboCar, parameterName=TurboCar_GlowBK_Radius}
    B=LinearInterpolate(A=@n19, B=@n20, Alpha=@n21.G)
    Alpha=@n21.R
   )
   B=ScalarParameter() {defaultValue=0.5, group=Wrap_TurboCar, parameterName=TurboCar_GlowBK_Radius_Mech}
   Alpha=@n21.B
  )
  n23 = ScalarParameter() {defaultValue=0.25, group=Wrap_TurboCar, parameterName=TurboCar_GlowBK_Hardness}
  n24 = TextureCoordinate()
  n25 = Saturate(None=MaterialFunctionCall() {materialFunction=&ObjectScale}.Scale Y)
  n26 = ScalarParameter() {defaultValue=1, group=Wrap_TurboCar, parameterName=TurboCar_Layers_Edge_Depth_Amount_Vehicle}
  n27 = ScalarParameter() {defaultValue=1, group=Wrap_TurboCar, parameterName=TurboCar_Layers_Edge_Depth_Amount_VehicleSide}
  n28 = MaterialFunctionCall(Color=@n9.RGB) {materialFunction=&SplitComponents}
  n29 = LinearInterpolate(
   A=LinearInterpolate(
    A=ScalarParameter() {defaultValue=10, group=Wrap_TurboCar, parameterName=TurboCar_Layers_Edge_Depth_Amount}
    B=LinearInterpolate(A=@n26, B=@n27, Alpha=@n28.G)
    Alpha=@n28.R
   )
   B=ScalarParameter() {defaultValue=1, group=Wrap_TurboCar, parameterName=TurboCar_Layers_Edge_Depth_Amount_Mech}
   Alpha=@n28.B
  )
  n30 = CameraVectorWS()
  n31 = ComponentMask(None=Multiply(A=Multiply(A=@n25, B=@n29), B=Transform(None=@n30) {transformSourceType=TRANSFORMSOURCE_World, transformType=TRANSFORM_Tangent})) {g=true, r=true}
  n32 = Multiply(A=@n31, B=Constant() {r=-0.003})
  n33 = TextureObjectParameter() {group=Customization, parameterName="CustomizationMask (_CM)", sortPriority=-400, texture=&T_BlacksRGB}
  n34 = Add(A=TextureSample(UVs=Add(A=@n24, B=@n32), Tex=@n33).G) {constB=-0.5}
  n35 = ScalarParameter() {defaultValue=8, group=Wrap_TurboCar, parameterName=TurboCar_Curvature_Mask B_Vehicle}
  n36 = ScalarParameter() {defaultValue=8, group=Wrap_TurboCar, parameterName=TurboCar_Curvature_Mask B_VehicleSide}
  n37 = MaterialFunctionCall(Color=@n9.RGB) {materialFunction=&SplitComponents}
  n38 = LinearInterpolate(
   A=LinearInterpolate(
    A=ScalarParameter() {defaultValue=8, group=Wrap_TurboCar, parameterName=TurboCar_Curvature_Mask B}
    B=LinearInterpolate(A=@n35, B=@n36, Alpha=@n37.G)
    Alpha=@n37.R
   )
   B=ScalarParameter() {defaultValue=8, group=Wrap_TurboCar, parameterName=TurboCar_Curvature_Mask B_Mech}
   Alpha=@n37.B
  )
  n39 = ScalarParameter() {group=Wrap_TurboCar, parameterName=TurboCar_Curvature_Mask A_Vehicle}
  n40 = ScalarParameter() {group=Wrap_TurboCar, parameterName=TurboCar_Curvature_Mask A_VehicleSide}
  n41 = MaterialFunctionCall(Color=@n9.RGB) {materialFunction=&SplitComponents}
  n42 = LinearInterpolate(
   A=LinearInterpolate(
    A=ScalarParameter() {group=Wrap_TurboCar, parameterName=TurboCar_Curvature_Mask A}
    B=LinearInterpolate(A=@n39, B=@n40, Alpha=@n41.G)
    Alpha=@n41.R
   )
   B=ScalarParameter() {group=Wrap_TurboCar, parameterName=TurboCar_Curvature_Mask A_Mech}
   Alpha=@n41.B
  )
  n43 = Saturate(None=LinearInterpolate(A=@n42, B=@n38, Alpha=Saturate(None=@n34)))
  n44 = Multiply(A=@n31, B=Constant() {r=-0.002})
  n45 = Add(A=TextureSample(UVs=Add(A=@n24, B=@n44), Tex=@n33).G) {constB=-0.5}
  n46 = Saturate(None=LinearInterpolate(A=@n42, B=@n38, Alpha=Saturate(None=@n45)))
  n47 = Multiply(A=@n31, B=Constant() {r=-0.001})
  n48 = Add(A=TextureSample(UVs=Add(A=@n24, B=@n47), Tex=@n33).G) {constB=-0.5}
  n49 = Saturate(None=LinearInterpolate(A=@n42, B=@n38, Alpha=Saturate(None=@n48)))
  n50 = LinearInterpolate(A=LinearInterpolate(A=Multiply(A=@n43) {constB=0.3}, Alpha=@n46) {constB=0.7}, Alpha=@n49) {constB=0.9}
  n51 = Add(A=TextureSample(UVs=@n24, Tex=@n33).G) {constB=-0.5}
  n52 = Saturate(None=LinearInterpolate(A=@n42, B=@n38, Alpha=Saturate(None=@n51)))
  n53 = ScalarParameter() {defaultValue=0.75, group=Wrap_TurboCar, parameterName=TurboCar_Glow_Effect}
  n54 = SphereMask(
   A=@n17
   B=@n18
   Radius=ScalarParameter() {defaultValue=0.5, group=Wrap_TurboCar, parameterName=TurboCar_Glow_Radius}
   Hardness=ScalarParameter() {defaultValue=0.25, group=Wrap_TurboCar, parameterName=TurboCar_Glow_Hardness}
  )
  n55 = Saturate(None=Multiply(A=LinearInterpolate(A=@n50, Alpha=@n52), B=LinearInterpolate(B=@n53, Alpha=@n54) {constA=1, constB=0.75}))
  n56 = CurveAtlasRowParameter(CurveTime=@n55) {
   atlas=/Game/Weapons/WeaponSkins/Wraps/Textures/CruveAtlas_Wraps_01
   curve=/Game/Weapons/WeaponSkins/Wraps/Textures/TurboC/CC_TurboC_A
   defaultValue=45
   group=Wrap_TurboCar
   parameterName=TurboCar_Edge_Color
  }
  n57 = LinearInterpolate(
   A=LinearInterpolate(
    A=VectorParameter() {group=Wrap_TurboCar, parameterName=TurboCar_GlowBK_Color A, defaultValue=<rgba 0 0 0 1>}.RGB
    B=VectorParameter() {group=Wrap_TurboCar, parameterName=TurboCar_GlowBK_Color B, defaultValue=<rgba 0 0 0 1>}.RGB
    Alpha=SphereMask(A=@n17, B=@n18, Radius=@n22, Hardness=@n23)
   )
   B=MaterialFunctionCall!(
    Index=ScalarParameter() {defaultValue=45, group=Wrap_TurboCar, parameterName="TurboCar_Edge_Color  Index"}
    CCinput=@n56
    CCinput_Alpha
    UV=@n55
   ) {materialFunction=&/MF_Global_CC_Tex}.3V
   Alpha=@n55
  )
  n58 = MaterialFunctionCall(Color=@n9.RGB) {materialFunction=&SplitComponents}
  n59 = LinearInterpolate(
   A=ScalarParameter() {defaultValue=10, group=Wrap_TurboCar, parameterName=TurboCar_Layers_Mask_Depth_Amount}
   B=LinearInterpolate(
    A=ScalarParameter() {defaultValue=1, group=Wrap_TurboCar, parameterName=TurboCar_Layers_Mask_Depth_Amount_Vehicle}
    B=ScalarParameter() {defaultValue=1, group=Wrap_TurboCar, parameterName=TurboCar_Layers_Mask_Depth_Amount_VehicleSide}
    Alpha=@n58.G
   )
   Alpha=@n58.R
  )
  n60 = ScalarParameter() {defaultValue=1, group=Wrap_TurboCar, parameterName=TurboCar_Layers_Mask_Depth_Amount_Mech}
  n61 = Multiply(
   A=Transform(None=@n30) {transformSourceType=TRANSFORMSOURCE_World, transformType=TRANSFORM_Local}
   B=Multiply(A=@n25, B=LinearInterpolate(A=@n59, B=@n60, Alpha=@n58.B))
  )
  n62 = Multiply(A=@n61) {constB=-0.004}
  n63 = PreSkinnedLocalBounds()
  n64 = FunctionInput(Preview=Divide(A=Subtract(A=@n1.PS, B=@n63.Min), B=Subtract(A=@n63.Max, B=@n63.Max))) {bUsePreviewValueAsDefault=true, inputName=Calcated_Bounds, sortPriority=11}
  n65 = MaterialFunctionCall(Color=@n9.RGB) {materialFunction=&SplitComponents}
  n66 = VectorParameter() {group=Wrap_TurboCar, parameterName=TurboCar_Mask_Offset_Vehicle, defaultValue=<rgba 0 0 0 1>}
  n67 = VectorParameter() {group=Wrap_TurboCar, parameterName=TurboCar_Mask_Offset_VehicleSide, defaultValue=<rgba 0 0 0 1>}
  n68 = MaterialFunctionCall(Color=@n9.RGB) {materialFunction=&SplitComponents}
  n69 = LinearInterpolate(
   A=LinearInterpolate(
    A=VectorParameter() {group=Wrap_TurboCar, parameterName=TurboCar_Mask_Offset_Weapon, defaultValue=<rgba 0 0 0 1>}.RGB
    B=LinearInterpolate(A=@n66.RGB, B=@n67.RGB, Alpha=@n68.G)
    Alpha=@n68.R
   )
   B=VectorParameter() {group=Wrap_TurboCar, parameterName=TurboCar_Mask_Offset_Mech, defaultValue=<rgba 0 0 0 1>}.RGB
   Alpha=@n68.B
  )
  n70 = ScalarParameter() {defaultValue=1, group=Wrap_TurboCar, parameterName=TurboCar_Mask_Scale_Vehicle}
  n71 = ScalarParameter() {defaultValue=1, group=Wrap_TurboCar, parameterName=TurboCar_Mask_Scale_VehicleSide}
  n72 = MaterialFunctionCall(Color=@n9.RGB) {materialFunction=&SplitComponents}
  n73 = LinearInterpolate(
   A=LinearInterpolate(
    A=ScalarParameter() {defaultValue=1, group=Wrap_TurboCar, parameterName=TurboCar_Mask_Scale_Weapon}
    B=LinearInterpolate(A=@n70, B=@n71, Alpha=@n72.G)
    Alpha=@n72.R
   )
   B=ScalarParameter() {defaultValue=1, group=Wrap_TurboCar, parameterName=TurboCar_Mask_Scale_Mech}
   Alpha=@n72.B
  )
  n74 = ScalarParameter() {group=Wrap_TurboCar, parameterName=TurboCar_Mask_Rotation_Vehicle}
  n75 = ScalarParameter() {group=Wrap_TurboCar, parameterName=TurboCar_Mask_Rotation_VehicleSide}
  n76 = MaterialFunctionCall(Color=@n9.RGB) {materialFunction=&SplitComponents}
  n77 = LinearInterpolate(
   A=LinearInterpolate(
    A=ScalarParameter() {group=Wrap_TurboCar, parameterName=TurboCar_Mask_Rotation_Weapon}
    B=LinearInterpolate(A=@n74, B=@n75, Alpha=@n76.G)
    Alpha=@n76.R
   )
   B=ScalarParameter() {defaultValue=0.25, group=Wrap_TurboCar, parameterName=TurboCar_Mask_Rotation_Mech}
   Alpha=@n76.B
  )
  n78 = FunctionInput(Preview=VertexInterpolator(VS=PreSkinnedNormal()).PS) {bUsePreviewValueAsDefault=true, inputName=Pre-Skinned Local Normal, sortPriority=10}
  n79 = TextureObjectParameter() {group=Wrap_TurboCar, parameterName=TurboCar_Mask__Tex, texture=&T_BlacksRGB}
  n80 = TextureSample(
   UVs=MaterialFunctionCall!(
    For_Safe_Zones
    Normal_Map
    Caculated_Bounds=Add(A=@n64, B=@n62)
    Pre_Skinned_Normals=@n78
    Is Vehicle=@n65.R
    Is VehicleSide=@n65.R
    Is Mech=@n65.R
    Vehicle_Rotation
    UV_Rotation Bypass=@n77
    Mech_Rotation
    Uvs_Offset=@n69
    Offset-after rotation
    Uvs_TileRate=@n73
   ) {materialFunction=&/MF_Projection_Side_Normal}.UVs
   Tex=@n79
  )
  n81 = Multiply(A=@n61) {constB=-0.003}
  n82 = TextureSample(
   UVs=MaterialFunctionCall(
    Caculated_Bounds=Add(A=@n64, B=@n81)
    Pre_Skinned_Normals=@n78
    Is Vehicle=@n65.R
    Is VehicleSide=@n65.R
    Is Mech=@n65.R
    UV_Rotation Bypass=@n77
    Uvs_Offset=@n69
    Uvs_TileRate=@n73
   ) {materialFunction=&MF_Projection_Side_Normal}.UVs
   Tex=@n79
  )
  n83 = LinearInterpolate(A=Multiply(A=@n80.R) {constB=0.2}, Alpha=@n82.R) {constB=0.4}
  n84 = Multiply(A=@n61) {constB=-0.002}
  n85 = TextureSample(
   UVs=MaterialFunctionCall(
    Caculated_Bounds=Add(A=@n64, B=@n84)
    Pre_Skinned_Normals=@n78
    Is Vehicle=@n65.R
    Is VehicleSide=@n65.R
    Is Mech=@n65.R
    UV_Rotation Bypass=@n77
    Uvs_Offset=@n69
    Uvs_TileRate=@n73
   ) {materialFunction=&MF_Projection_Side_Normal}.UVs
   Tex=@n79
  )
  n86 = Multiply(A=@n61) {constB=-0.001}
  n87 = TextureSample(
   UVs=MaterialFunctionCall(
    Caculated_Bounds=Add(A=@n64, B=@n86)
    Pre_Skinned_Normals=@n78
    Is Vehicle=@n65.R
    Is VehicleSide=@n65.R
    Is Mech=@n65.R
    UV_Rotation Bypass=@n77
    Uvs_Offset=@n69
    Uvs_TileRate=@n73
   ) {materialFunction=&MF_Projection_Side_Normal}.UVs
   Tex=@n79
  )
  n88 = MaterialFunctionCall(
   Caculated_Bounds=@n64
   Pre_Skinned_Normals=@n78
   Is Vehicle=@n65.R
   Is VehicleSide=@n65.R
   Is Mech=@n65.R
   UV_Rotation Bypass=@n77
   Uvs_Offset=@n69
   Uvs_TileRate=@n73
  ) {materialFunction=&MF_Projection_Side_Normal}
  n89 = TextureSample(UVs=@n88.UVs, Tex=@n79)
  n90 = LinearInterpolate(A=LinearInterpolate(A=LinearInterpolate(A=@n83, Alpha=@n85.R) {constB=0.6}, Alpha=@n87.R) {constB=0.8}, Alpha=@n89.R)
  n91 = Saturate(None=@n90)
  n92 = CurveAtlasRowParameter(CurveTime=@n91) {
   atlas=&CruveAtlas_Wraps_01
   curve=/Game/Weapons/WeaponSkins/Wraps/Textures/TurboC/CC_TurboC_B
   defaultValue=46
   group=Wrap_TurboCar
   parameterName=TurboCar_Mask_Color
  }
  n93 = MaterialFunctionCall(Index=ScalarParameter() {defaultValue=45, group=Wrap_TurboCar, parameterName="TurboCar_Mask_Color  Index"}, CCinput=@n92, UV=@n91) {materialFunction=&MF_Global_CC_Tex}
  Add(A=@n57, B=@n93.3V)
  n94 = FunctionInput() {bUsePreviewValueAsDefault=true, inputName=Material, inputType=FunctionInput_MaterialAttributes, sortPriority=-20}
  n95 = GetMaterialAttributes(None=@n94) {attributeGetTypes [#1]}
  n96 = VectorParameter() {group=Wrap_TurboCar, parameterName=TurboCar_BaseColor, sortPriority=0, defaultValue=<rgba 0 0 0 1>}
  n97 = MaterialFunctionCall(Color=Abs(None=@n78)) {materialFunction=&SplitComponents}
  n98 = LinearInterpolate(A=@n97.B, B=@n97.B, Alpha=@n10.G)
  n99 = MaterialFunctionCall(
   In=LinearInterpolate(A=LinearInterpolate(A=@n97.R, B=@n98, Alpha=@n10.R), B=@n97.G, Alpha=@n10.B)
   Contrast=ScalarParameter() {group=Wrap_TurboCar, parameterName=TurboCar_Mask__Triplanner_Mask_Contrast}
  ) {materialFunction=&CheapContrast}
  n100 = MaterialFunctionCall(Color=@n9.RGB) {materialFunction=&SplitComponents}
  n101 = LinearInterpolate(
   A=VectorParameter() {group=Wrap_TurboCar, parameterName=TurboCar_MaskClamp_Offset_Weapon, defaultValue=<rgba 0 0 0 1>}.RGB
   B=LinearInterpolate(
    A=VectorParameter() {group=Wrap_TurboCar, parameterName=TurboCar_MaskClamp_Offset_Vehicle, defaultValue=<rgba 0 0 0 1>}.RGB
    B=VectorParameter() {group=Wrap_TurboCar, parameterName=TurboCar_MaskClamp_Offset_VehicleSide, defaultValue=<rgba 0 0 0 1>}.RGB
    Alpha=@n100.G
   )
   Alpha=@n100.R
  )
  n102 = VectorParameter() {group=Wrap_TurboCar, parameterName=TurboCar_MaskClamp_Offset_Mech, defaultValue=<rgba 0 0 0 1>}
  n103 = Add(A=@n88.UVs, B=ComponentMask(None=LinearInterpolate(A=@n101, B=@n102.RGB, Alpha=@n100.B)) {g=true, r=true})
  n104 = VectorParameter() {group=Wrap_TurboCar, parameterName=TurboCar_MaskClamp_Scale_Weapon, defaultValue=<rgba 1 1 0 1>}
  n105 = MaterialFunctionCall(Color=@n9.RGB) {materialFunction=&SplitComponents}
  n106 = LinearInterpolate(
   A=VectorParameter() {group=Wrap_TurboCar, parameterName=TurboCar_MaskClamp_Scale_Vehicle, defaultValue=<rgba 1 1 0 1>}.RGB
   B=VectorParameter() {group=Wrap_TurboCar, parameterName=TurboCar_MaskClamp_Scale_VehicleSide, defaultValue=<rgba 1 1 0 1>}.RGB
   Alpha=@n105.G
  )
  n107 = ComponentMask(None=LinearInterpolate(
    A=LinearInterpolate(A=@n104.RGB, B=@n106, Alpha=@n105.R)
    B=VectorParameter() {group=Wrap_TurboCar, parameterName=TurboCar_MaskClamp_Scale_Mech, defaultValue=<rgba 1 1 0 1>}.RGB
    Alpha=@n105.B
   )) {g=true, r=true}
  n108 = Multiply(
   A=Multiply(A=@n99.Result, B=@n90)
   B=MaterialFunctionCall(UVs=@n103, Texture Scale=@n107) {materialFunction=&ScaleUVsByCenter}.0-1 mask
  )
  n109 = Constant() {r=0.5}
  n110 = Constant() {r=1}
  n111 = Constant() {r=0.5}
  n112 = Constant() {r=0.25}
  n113 = SetMaterialAttributes!(
   MaterialAttributes=@n95.MaterialAttributes
   BaseColor=@n96.RGB
   Roughness=ScalarParameter() {defaultValue=0.2, group=Wrap_TurboCar, parameterName=TurboCar_Roghness, sortPriority=0}
   Specular=ScalarParameter() {defaultValue=0.5, group=Wrap_TurboCar, parameterName=TurboCar_Specular, sortPriority=0}
   EmissiveColor=Multiply(
    A=LinearInterpolate(A=@n57, B=@n93.3V, Alpha=@n108)
    B=MaterialFunctionCall(Value_Morning=@n109, Value_Day=@n110, Value_Evening=@n111, Value_Night=@n112) {materialFunction=&Scalar_Time_of_Day}.Result
   )
   Metallic=ScalarParameter() {group=Wrap_TurboCar, parameterName=TurboCar_Metallic, sortPriority=0}
  ) {attributeSetTypes [
    #2
    #3
    #4
    #5
    #6
   ]}
  n114 = FunctionInput() {bUsePreviewValueAsDefault=true, inputName="Customization Mask (_CM B)", inputType=FunctionInput_Scalar, sortPriority=24}
  n115 = ScalarParameter() {group=Wrap_TurboCar, parameterName=TurboCar_Remove_Customization Mask, sortPriority=-10}
  BlendMaterialAttributes(A=@n95.MaterialAttributes, B=@n113, Alpha=Saturate(None=Add(A=@n114, B=@n115))) {vertexAttributeBlendType=UseA}
  FunctionInput() {bUsePreviewValueAsDefault=true, inputName=Curavture, inputType=FunctionInput_Scalar, sortPriority=24}
  FunctionInput(Preview=Constant3Vector() {constant=<rgba 0 1 0 0>}) {bUsePreviewValueAsDefault=true, inputName=ForwardAxis}
  FunctionInput(Preview=Constant3Vector() {constant=<rgba 0.3918 0 0.145 1>}) {bUsePreviewValueAsDefault=true, inputName=Object Bounds Center Adjustment, sortPriority=50}
  FunctionOutput(None=StaticSwitchParameter(True=@n113, False=@n94) {group=Customization_Wrap, parameterName=Wrap_TurboCar})
  n116 = Multiply(A=@n31, B=Constant() {r=-0.004})
  n117 = Add(A=TextureSample(UVs=Add(A=@n24, B=@n116), Tex=@n33).G) {constB=-0.5}
  n118 = Saturate(None=LinearInterpolate(A=@n42, B=@n38, Alpha=Saturate(None=@n117)))
  n119 = Multiply(A=@n118) {constB=0.2}
  LinearInterpolate(A=@n119, Alpha=@n43) {constB=0.5}
  n120 = MaterialFunctionCall(
   Caculated_Bounds=Add(A=@n64, B=Multiply(A=@n61) {constB=-0.005})
   Pre_Skinned_Normals=@n78
   Is Vehicle=@n65.R
   Is VehicleSide=@n65.R
   Is Mech=@n65.R
   Uvs_TileRate=@n73
  ) {materialFunction=&MF_Projection_Side_Normal}
  LinearInterpolate(A=Multiply(A=TextureSample(UVs=@n120.UVs, Tex=@n79).R) {constB=0.1}, Alpha=@n80.R) {constB=0.2}
  LinearInterpolate(A=@n119, Alpha=@n118) {constB=0.3}
  n121 = Constant() {r=0.5}
  n122 = MaterialFunctionCall(Float2=Multiply(A=@n103, B=@n107)) {materialFunction=&BreakOutFloat2Components}
  Multiply(A=SphereMask(A=@n121, B=@n122.R) {attenuationRadius=0.5}, B=SphereMask(A=@n122.G, B=@n121) {attenuationRadius=0.5})
  Multiply(A=Multiply(A=@n99.Result, B=@n93.3V))
  Multiply(A=@n89.R)
  Multiply(A=@n87.R) {constB=0.75}
  Multiply(A=@n85.R) {constB=0.5}
  Multiply(A=@n82.R) {constB=0.25}
  n123 = FunctionInput(Preview=Constant4Vector() {constant=<rgba 0 0 0 1>}.RGBA) {bUsePreviewValueAsDefault=true, inputName=Object-SafeZone-Virtual-Pivot, inputType=FunctionInput_Vector4, sortPriority=105}
  n124 = FunctionInput() {bUsePreviewValueAsDefault=true, inputName=Object-SafeZone-Virtual-CH-Switch, inputType=FunctionInput_Scalar, sortPriority=105}
  NamedRerouteDeclaration(None=MaterialFunctionCall!(
    Camera _pos
    Axis
    Scale
    Scale_Multplyer
    Object-SafeZone-Virtual-Pivot=@n123
    Object-SafeZone-Virtual-CH-Switch=@n124
    SinglePlane Normal Axis
    SinglePlane X Axis
    PlaneCenter_distortion
   ) {materialFunction=&/MF_SafeZone_VirtualPlane}.Uvs) {name=SafeZoneVP_UVs, nodeColor=<rgba 1 0 0.6562 1>}
  NamedRerouteDeclaration(None=@n8.A) {name=IsCharacter, nodeColor=<rgba 0 1 0.1797 1>}
  Round(None=@n52)
  Round(None=@n49)
  Round(None=@n46)
  Round(None=@n43)
  Round(None=@n118)
  Saturate(None=Multiply(A=@n90) {constB=4})
  TextureSampleParameter2D() {
   group=Customization
   parameterName="CustomizationMask (_CM)"
   samplerSource=SSM_Wrap_WorldGroupSettings
   samplerType=SAMPLERTYPE_Masks
   sortPriority=-400
   texture=&T_BlackMask
  }
 ]}
