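# Material function for the Veiled Glitch wrap skin, applied over an incoming material through masks and a static switch.
# Pre-skinned bounds and normals drive projected UVs, with IsState inputs selecting parameters for weapons, vehicles, mechs, characters, backpacks and pickaxes.
# It builds textured A and animated emissive B layers, the latter using warped noise, curve-atlas colors, gridlines, Fresnel and time-of-day scaling.
# A timestamp-driven reactive section adds pixelated gradient masks and glow.
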MaterialFunction /Game/Weapons/WeaponSkins/Wraps/MaterialFunctions/MF_Wrap_VGlitch
graph {nodes [
  n1 = NamedRerouteUsage()
  Add(
   A=NamedRerouteUsage()
   B=Multiply(
    A=ComponentMask(None=@n1) {g=true, r=true}
    B=ScalarParameter() {defaultValue=0.05, group=Wrap_VeiledGlitch_B_Noise, parameterName=VGlit_BC_Warp_Strength}
   ) {constB=0.025}
  )
  n2 = Multiply(A=NamedRerouteUsage(), B=ScalarParameter() {defaultValue=0.5, group=Wrap_VeiledGlitch_B_Grid, parameterName=VGlit_B_Grad_TileRate})
  n3 = Multiply(A=Time(), B=ScalarParameter() {defaultValue=-0.25, group=Wrap_VeiledGlitch_B_Grid, parameterName=VGlit_B_Grad_Speed})
  n4 = Frac(None=ComponentMask(None=Add(A=@n2, B=@n3)) {r=true})
  n5 = CurveAtlasRowParameter(CurveTime=@n4) {
   atlas=/Game/Weapons/WeaponSkins/Wraps/Textures/CruveAtlas_Wraps_01
   curve=/Game/Weapons/WeaponSkins/Wraps/Textures/V_Glitch/CC_VGlitch_Msk
   defaultValue=102
   group=Wrap_VeiledGlitch_B_Grid
   parameterName=VGlit__CC_Grad
  }
  MaterialExpressionMaterialFunctionCall_16 = MaterialFunctionCall!(
   Index=ScalarParameter() {defaultValue=102, group=Wrap_VeiledGlitch_B_Grid, parameterName="VGlit__CC_Grad  Index"}
   CCinput=@n5
   CCinput_Alpha
   UV=@n4
  ) {materialFunction=&/MF_Global_CC_Tex}
  n6 = Convert(None=@MaterialExpressionMaterialFunctionCall_16.3V) {
   nodeName=Break Float3
   convertInputs [{
     type=Vector3
     expressionInput {
      expression=&self:MaterialExpressionMaterialFunctionCall_16
      inputName=None
      mask=0
      maskA=0
      maskB=0
      maskG=0
      maskR=0
      outputIndex=0
     }
    }]
   convertOutputs [{type=Scalar}, {type=Scalar}, {type=Scalar}]
  }
  Add(A=@n6)
  n7 = NamedRerouteUsage()
  n8 = LinearInterpolate(
   A=ScalarParameter() {defaultValue=1.03, group=Wrap_VeiledGlitch_Reactive, parameterName=VGlit_Reactive_Postion_Start}
   B=ScalarParameter() {defaultValue=0.03, group=Wrap_VeiledGlitch_Reactive, parameterName=VGlit_Reactive_Position_End}
   Alpha=@n7
  )
  Add(A=@n8)
  n9 = NamedRerouteUsage()
  Add(
   A=@n9
   B=Multiply(A=NamedRerouteUsage(), B=ScalarParameter() {defaultValue=0.05, group=Wrap_VeiledGlitch_B, parameterName=VGlit_Accent_B_Strength})
  )
  n10 = ScalarParameter() {defaultValue=10, group=Wrap_VeiledGlitch_B, parameterName=VGlit_B_Grid_Brightness}
  n11 = ComponentMask(None=NamedRerouteUsage()) {r=true}
  n12 = CurveAtlasRowParameter(CurveTime=@n11) {
   atlas=&CruveAtlas_Wraps_01
   curve=/Game/Weapons/WeaponSkins/Wraps/Textures/V_Glitch/CC_VGlitch_Bc
   defaultValue=101
   group=Wrap_VeiledGlitch_B
   parameterName=VGlit__CC_BC
  }
  n13 = MaterialFunctionCall(
   Index=ScalarParameter() {defaultValue=101, group=Wrap_VeiledGlitch_B, parameterName="VGlit__CC_BC  Index"}
   CCinput=@n12
   CCinput_Alpha=@n12
   UV=@n11
  ) {materialFunction=&MF_Global_CC_Tex}
  n14 = VectorParameter() {group=Wrap_VeiledGlitch_B_Grid, parameterName=VGlit_Grid_ColorTint, defaultValue=<rgba 0.3629 0.176 0.5 1>}
  n15 = Multiply(A=NamedRerouteUsage(), B=LinearInterpolate(A=@n13.3V, B=@n14.RGB, Alpha=@n13.3V))
  n16 = VectorParameter() {group=Wrap_VeiledGlitch_B_Grid, parameterName=VGlit_B_Rim_ColorTint, defaultValue=<rgba 0.6325 0.3068 0.8714 0.02>}
  n17 = ScalarParameter() {defaultValue=10, group=Wrap_VeiledGlitch_B_Noise, parameterName=VGlit_B_Fresnel_Expo}
  Add(A=Add(A=Multiply(A=@n10, B=@n15), B=NamedRerouteUsage()), B=Multiply(A=@n16.RGB, B=Fresnel(ExponentIn=@n17) {baseReflectFraction=0}))
  Constant() {r=1}
  Constant() {r=0.5}
  Constant() {r=0.25}
  Constant() {r=1}
  Constant() {r=0.5}
  Constant() {r=1}
  n18 = VertexInterpolator(VS=PreSkinnedPosition())
  n19 = PreSkinnedLocalBounds()
  n20 = MaterialFunctionCall(Color=Subtract(A=@n19.Max, B=@n19.Max)) {materialFunction=&SplitComponents}
  n21 = Max(A=@n20.R, B=@n20.R)
  n22 = FunctionInput(Preview=Divide(A=Subtract(A=@n18.PS, B=@n19.Min), B=Max(A=@n21, B=@n20.B))) {bUsePreviewValueAsDefault=true, inputName=Calcated_Bounds, sortPriority=11}
  n23 = ScalarParameter() {defaultValue=10, group=Wrap_VeiledGlitch_Reactive, parameterName=VGlit_Reactive_Pixel_Density_B}
  n24 = ScalarParameter() {defaultValue=1000, group=Wrap_VeiledGlitch_Reactive, parameterName=VGlit_Reactive_Pixel_Density_A}
  n25 = LinearInterpolate(A=@n23, B=@n24, Alpha=@n7)
  Divide(A=Floor(None=Multiply(A=@n22, B=@n25)), B=@n25)
  Divide(A=ComponentMask(None=NamedRerouteUsage()) {b=true}) {constB=150}
  n26 = MaterialFunctionCall!(
   Normal_Map
   Offset Rt
   Pivot
   Caculated_Bounds=@n22
   Pre_Skinned_Normals=NamedRerouteUsage()
   is_State=NamedRerouteUsage()
   UV_Rotation=NamedRerouteUsage()
   Uvs_Offset=@n8
   Uvs_TileRate=ScalarParameter() {defaultValue=1, group=Wrap_VeiledGlitch_Reactive, parameterName=VGlit_Reactive_TileRate}
   Uvs_Offset_Pivot
   For_Safe_Zones
   Rotation_Comp
   Clamp Y
   Clamp X
   Clamp Padding_Comp
   No_Rotation
  ) {materialFunction=&/MF_Projection_Side_Normal_V2}
  n27 = ComponentMask(None=@n26.UVs) {r=true}
  n28 = ScalarParameter() {defaultValue=20, group=Wrap_VeiledGlitch_Reactive, parameterName=VGlit_Reactive_Steps_A}
  n29 = Divide(A=Floor(None=Multiply(A=@n27, B=@n28)), B=@n28)
  Frac(None=@n29)
  n30 = NamedRerouteUsage()
  n31 = NamedRerouteUsage()
  n32 = TextureSample() {samplerType=SAMPLERTYPE_Masks, texture=/Game/Weapons/FORT_Rifles/Textures/T_Scar_CM}
  n33 = FunctionInput(Preview=@n32.R) {bUsePreviewValueAsDefault=true, inputName=Mask Primary, inputType=FunctionInput_Scalar, sortPriority=24, previewValue=<1 0 0 1>}
  n34 = FunctionInput(Preview=@n32.B) {bUsePreviewValueAsDefault=true, inputName=Mask Secondary, inputType=FunctionInput_Scalar, sortPriority=24, previewValue=<1 0 0 1>}
  n35 = Saturate(None=Add(A=@n33, B=@n34))
  n36 = TextureSample() {samplerType=SAMPLERTYPE_Normal, texture=/Game/Weapons/FORT_Rifles/T_SCAR_N}
  n37 = Constant3Vector() {constant=<rgba 1 0 0 0>}
  n38 = GetMaterialAttributes(None=FunctionInput(Preview=SetMaterialAttributes!(MaterialAttributes, Normal=@n36.RGB, BaseColor=@n37) {attributeSetTypes [#1, #2]}) {inputName=Material, inputType=FunctionInput_MaterialAttributes}) {attributeGetTypes [
    #2
    #3
    #1
    #4
   ]}
  FunctionOutput(None=StaticSwitchParameter(
    True=BlendMaterialAttributes(A=@n30, B=@n31, Alpha=@n35) {vertexAttributeBlendType=UseA}
    False=NamedRerouteDeclaration(None=@n38.MaterialAttributes) {name=Root_Material, nodeColor=<rgba 0.7812 1 0 1>}
   ) {group=Customization_Wrap, parameterName=Wrap_VeiledGlitch})
  LinearInterpolate(A=@n24, B=@n23)
  LinearInterpolate(Alpha=@n7)
  n39 = NamedRerouteUsage()
  MaterialFunctionCall(Float3=@n39) {materialFunction=&BreakOutFloat3Components}
  MaterialFunctionCall!(Gridlines, Noise, Fresnel, In) {materialFunction=/Game/Characters/Player/Female/Medium/Bodies/F_MED_Veiled_Glitch/Materials/M_VeiledGlitch_Gridlines}
  Multiply(A=NamedRerouteUsage(), B=@n13.3V)
  Multiply(A=Add(A=@n13.3V, B=@n16.RGB), B=ScalarParameter() {defaultValue=1, group=Wrap_VeiledGlitch_B, parameterName=VGlit_B_Rim_Brightness})
  Multiply()
  Multiply(A=MaterialFunctionCall() {materialFunction=&Scalar_Time_of_Day}.Result)
  Multiply(A=MaterialFunctionCall!(X range Y range, Gradient Height, ShowGrid?, Thickness, Color, "f(x)") {materialFunction=/Engine/Functions/Engine_MaterialFunctions01/Debug/PlotFunctionOnGraph}."f(x) = y")
  n40 = ScalarParameter() {defaultValue=75, group=Wrap_VeiledGlitch_B_Noise, parameterName=VGlit_B_Noise_B_Size}
  n41 = MaterialFunctionCall(Float4=NamedRerouteUsage()) {materialFunction=&BreakOutFloat4Components}
  n42 = LinearInterpolate(
   A=@n40
   B=LinearInterpolate(
    A=ScalarParameter() {defaultValue=50, group=Wrap_VeiledGlitch_B_Noise, parameterName=VGlit_B_Noise_B_Size Vehicle}
    B=ScalarParameter() {group=Wrap_VeiledGlitch_B_Noise, parameterName=VGlit_B_Noise_B_Size Vehicle Side}
    Alpha=@n41.G
   )
   Alpha=@n41.R
  )
  n43 = LinearInterpolate(A=ScalarParameter() {defaultValue=50, group=Wrap_VeiledGlitch_B_Noise, parameterName=VGlit_B_Noise_B_Size Mech}, B=@n40, Alpha=@n41.A)
  n44 = VectorParameter() {group=Wrap_VeiledGlitch_B_Noise, parameterName=VGlit_B_Noise_B_Speed, defaultValue=<rgba -0.15 0.05 0 1>}
  n45 = MaterialFunctionCall!(
   Texture=NamedRerouteUsage()
   Caculated_Bounds=Divide(A=NamedRerouteUsage(), B=LinearInterpolate(A=@n42, B=@n43, Alpha=@n41.B))
   Pre_Skinned_Normals=NamedRerouteUsage()
   Blend_Contrast=Constant()
   IsStatsB=NamedRerouteUsage()
   IsState=NamedRerouteUsage()
   Fill_alternate_Axis_Opacity=Constant()
   "Fill_alternate_Axis_Color_A=Roughness"=Constant4Vector().RGBA
   Offset=Multiply(A=Time(), B=@n44.RGB)
   Scale
   Offset_Pivot
   Rotation_Axis=NamedRerouteUsage()
  ) {materialFunction=&/MF_TopsSideProjection_Base_V2}
  Multiply(A=@n45.RGB)
  NamedRerouteDeclaration(None=@n22) {name=Caclcated_Bounds_B, nodeColor=<rgba 0 0.9297 1 1>}
  n46 = FunctionInput(Preview=@n32.G) {bUsePreviewValueAsDefault=true, inputName=Mask Crevise, inputType=FunctionInput_Scalar, sortPriority=24, previewValue=<1 0 0 1>}
  NamedRerouteDeclaration(None=@n46) {name=Msk_Curvature, nodeColor=<rgba 0 1 0.7422 1>}
  n47 = VectorParameter() {group=Wrap_VeiledGlitch_B_UVS, parameterName=VGlit_B_Rotation_Vehicle, defaultValue=<rgba 0 0 0 1>}
  n48 = VectorParameter() {group=Wrap_VeiledGlitch_B_UVS, parameterName=VGlit_B_Rotation_Vehicle_Side, defaultValue=<rgba 0 0 0 1>}
  n49 = MaterialFunctionCall(Float4=NamedRerouteUsage()) {materialFunction=&BreakOutFloat4Components}
  n50 = VectorParameter() {group=Wrap_VeiledGlitch_B_UVS, parameterName=VGlit_B_Rotation_Character, defaultValue=<rgba 0.25 0.25 0 1>}
  n51 = LinearInterpolate(
   A=LinearInterpolate(
    A=VectorParameter() {group=Wrap_VeiledGlitch_B_UVS, parameterName=VGlit_B_Rotation_Weapon, defaultValue=<rgba 0 0 -0.25 1>}.RGB
    B=LinearInterpolate(A=@n47.RGB, B=@n48.RGB, Alpha=@n49.G)
    Alpha=@n49.R
   )
   B=LinearInterpolate(
    A=VectorParameter() {group=Wrap_VeiledGlitch_B_UVS, parameterName=VGlit_B_Rotation_Mech, defaultValue=<rgba 0.25 0.25 0 1>}.RGB
    B=@n50.RGB
    Alpha=@n49.A
   )
   Alpha=@n49.B
  )
  n52 = VectorParameter() {group=Wrap_VeiledGlitch_B_UVS, parameterName=VGlit_B_Rotation_Backpack, defaultValue=<rgba 0 0 0 1>}
  n53 = MaterialFunctionCall(Float2=NamedRerouteUsage()) {materialFunction=&BreakOutFloat2Components}
  NamedRerouteDeclaration(None=LinearInterpolate(
    A=LinearInterpolate(A=@n51, B=@n52.RGB, Alpha=@n53.R)
    B=VectorParameter() {group=Wrap_VeiledGlitch_B_UVS, parameterName=VGlit_B_Rotation_Pickaxe, defaultValue=<rgba 0 0 0 1>}.RGB
    Alpha=@n53.G
   )) {name=UVs_Top_Side_Rotation_B, nodeColor=<rgba 0 0.7188 1 1>}
  n54 = ScalarParameter() {defaultValue=1, group=Wrap_VeiledGlitch_A_UVS, parameterName=VGlit_A_Scale_Vehicle, sortPriority=150}
  n55 = ScalarParameter() {defaultValue=1, group=Wrap_VeiledGlitch_A_UVS, parameterName=VGlit_A_Scale_Vehicle_Side, sortPriority=150}
  n56 = MaterialFunctionCall(Float4=NamedRerouteUsage()) {materialFunction=&BreakOutFloat4Components}
  n57 = LinearInterpolate(
   A=LinearInterpolate(
    A=ScalarParameter() {defaultValue=1, group=Wrap_VeiledGlitch_A_UVS, parameterName=VGlit_A_Scale_Weapon, sortPriority=150}
    B=LinearInterpolate(A=@n54, B=@n55, Alpha=@n56.G)
    Alpha=@n56.R
   )
   B=LinearInterpolate(
    A=ScalarParameter() {defaultValue=1, group=Wrap_VeiledGlitch_A_UVS, parameterName=VGlit_A_Scale_Mech, sortPriority=150}
    B=ScalarParameter() {defaultValue=1, group=Wrap_VeiledGlitch_A_UVS, parameterName=VGlit_A_Scale_Character, sortPriority=150}
    Alpha=@n56.A
   )
   Alpha=@n56.B
  )
  n58 = ScalarParameter() {defaultValue=1, group=Wrap_VeiledGlitch_A_UVS, parameterName=VGlit_A_Scale_Backpack, sortPriority=150}
  n59 = MaterialFunctionCall(Float2=NamedRerouteUsage()) {materialFunction=&BreakOutFloat2Components}
  NamedRerouteDeclaration(None=LinearInterpolate(
    A=LinearInterpolate(A=@n57, B=@n58, Alpha=@n59.R)
    B=ScalarParameter() {defaultValue=1, group=Wrap_VeiledGlitch_A_UVS, parameterName=VGlit_A_Scale_Pickaxe, sortPriority=150}
    Alpha=@n59.G
   )) {name=UVs_Scale_A, nodeColor=<rgba 0 0.7188 1 1>}
  n60 = Fresnel(ExponentIn=ScalarParameter() {defaultValue=1, group=Wrap_VeiledGlitch_B_Noise, parameterName=VGlit_B_Fresnel_Expo}) {baseReflectFraction=0}
  n61 = Time()
  n62 = VectorParameter() {group=Wrap_VeiledGlitch_B_Noise, parameterName=VGlit_B_Noise_Speed, defaultValue=<rgba 0.033 0.025 0 1>}
  n63 = Add(A=ComponentMask(None=Multiply(A=@n61, B=@n62.RGB)) {g=true, r=true}, B=NamedRerouteUsage())
  n64 = ComponentMask(None=Multiply(A=NamedRerouteUsage(), B=ScalarParameter() {defaultValue=0.1, group=Wrap_VeiledGlitch_B_Noise, parameterName=VGlit_B_Warp_Strength}) {constB=0}) {g=true, r=true}
  n65 = TextureSample(
   UVs=Multiply(A=ScalarParameter() {defaultValue=1.5, group=Wrap_VeiledGlitch_B_Noise, parameterName=VGlit_B_Noise_TileRate}, B=Add(A=@n63, B=@n64))
   Tex=NamedRerouteUsage()
  )
  NamedRerouteDeclaration(None=LinearInterpolate(A=@n60, Alpha=OneMinus(None=@n65.A))) {name=Glitch_Noise_Msk, nodeColor=<rgba 1 0 0.2344 1>}
  n66 = VectorParameter() {group=Wrap_VeiledGlitch_A_UVS, parameterName=VGlit_A_Offset_Vehicle, defaultValue=<rgba 0 0 0 1>}
  n67 = VectorParameter() {group=Wrap_VeiledGlitch_A_UVS, parameterName=VGlit_A_Offset_VehicleSide, defaultValue=<rgba 0 0 0 1>}
  n68 = MaterialFunctionCall(Float4=NamedRerouteUsage()) {materialFunction=&BreakOutFloat4Components}
  n69 = LinearInterpolate(
   A=LinearInterpolate(
    A=VectorParameter() {group=Wrap_VeiledGlitch_A_UVS, parameterName=VGlit_A_Offset_Weapon, defaultValue=<rgba 0 0 0 1>}.RGB
    B=LinearInterpolate(A=@n66.RGB, B=@n67.RGB, Alpha=@n68.G)
    Alpha=@n68.R
   )
   B=LinearInterpolate(
    A=VectorParameter() {group=Wrap_VeiledGlitch_A_UVS, parameterName=VGlit_A_Offset_Mech, defaultValue=<rgba 0 0 0 1>}.RGB
    B=VectorParameter() {group=Wrap_VeiledGlitch_A_UVS, parameterName=VGlit_A_Offset_Character, defaultValue=<rgba 0 0 0 1>}.RGB
    Alpha=@n68.A
   )
   Alpha=@n68.B
  )
  n70 = VectorParameter() {group=Wrap_VeiledGlitch_A_UVS, parameterName=VGlit_A_Offset_Backpack, defaultValue=<rgba 0 0 0 1>}
  n71 = MaterialFunctionCall(Float2=NamedRerouteUsage()) {materialFunction=&BreakOutFloat2Components}
  NamedRerouteDeclaration(None=LinearInterpolate(
    A=LinearInterpolate(A=@n69, B=@n70.RGB, Alpha=@n71.R)
    B=VectorParameter() {group=Wrap_VeiledGlitch_A_UVS, parameterName=VGlit_A_Offset_Pickaxe, defaultValue=<rgba 0 0 0 1>}.RGB
    Alpha=@n71.G
   )) {name=UVs_Offset_A, nodeColor=<rgba 0 0.625 1 1>}
  NamedRerouteDeclaration(None=TextureObjectParameter() {group=Wrap_VeiledGlitch_B, parameterName=VGlit__Tex_Msk, texture=&T_BlacksRGB}) {name=Tex_Msk, nodeColor=<rgba 0.0781 0 1 1>}
  n72 = TextureObjectParameter() {
   group=Wrap_VeiledGlitch_B
   parameterName=VGlit__Tex_N2
   samplerType=SAMPLERTYPE_Normal
   texture=&FlatNormalBlue
  }
  NamedRerouteDeclaration(None=@n72) {name=Tex_N2, nodeColor=<rgba 0.0781 0 1 1>}
  n73 = ScalarParameter() {group=Wrap_VeiledGlitch_A_UVS, parameterName=VGlit_A_Rotation_Vehicle, sortPriority=150}
  n74 = ScalarParameter() {group=Wrap_VeiledGlitch_A_UVS, parameterName=VGlit_A_Rotation_Vehicle_Side, sortPriority=150}
  n75 = MaterialFunctionCall(Float4=NamedRerouteUsage()) {materialFunction=&BreakOutFloat4Components}
  n76 = LinearInterpolate(
   A=LinearInterpolate(
    A=ScalarParameter() {group=Wrap_VeiledGlitch_A_UVS, parameterName=VGlit_A_Rotation_Weapon, sortPriority=150}
    B=LinearInterpolate(A=@n73, B=@n74, Alpha=@n75.G)
    Alpha=@n75.R
   )
   B=LinearInterpolate(
    A=ScalarParameter() {group=Wrap_VeiledGlitch_A_UVS, parameterName=VGlit_A_Rotation_Mech, sortPriority=150}
    B=ScalarParameter() {group=Wrap_VeiledGlitch_A_UVS, parameterName=VGlit_A_Rotation_Character, sortPriority=150}
    Alpha=@n75.A
   )
   Alpha=@n75.B
  )
  n77 = ScalarParameter() {group=Wrap_VeiledGlitch_A_UVS, parameterName=VGlit_A_Rotation_Backpack, sortPriority=150}
  n78 = MaterialFunctionCall(Float2=NamedRerouteUsage()) {materialFunction=&BreakOutFloat2Components}
  NamedRerouteDeclaration(None=LinearInterpolate(
    A=LinearInterpolate(A=@n76, B=@n77, Alpha=@n78.R)
    B=ScalarParameter() {group=Wrap_VeiledGlitch_A_UVS, parameterName=VGlit_A_Rotation_Pickaxe, sortPriority=150}
    Alpha=@n78.G
   )) {name=UVs_Rotation_A, nodeColor=<rgba 0 0.7188 1 1>}
  n79 = NamedRerouteUsage()
  n80 = MaterialFunctionCall(
   Normal_Map=NamedRerouteUsage()
   Caculated_Bounds=NamedRerouteUsage()
   Pre_Skinned_Normals=NamedRerouteUsage()
   is_State=NamedRerouteUsage()
   UV_Rotation=@n79
   Uvs_Offset=NamedRerouteUsage()
   Uvs_TileRate=NamedRerouteUsage()
  ) {materialFunction=&MF_Projection_Side_Normal_V2}
  n81 = NamedRerouteDeclaration(None=@n80.UVs) {name=Uvs, nodeColor=<rgba 0 0.3672 1 1>}
  n82 = TextureSample(UVs=@n81, Tex=NamedRerouteUsage()) {samplerType=SAMPLERTYPE_Normal}
  n83 = NamedRerouteUsage()
  n84 = SetMaterialAttributes!(
   MaterialAttributes=NamedRerouteUsage()
   BaseColor=@n82.RGB
   Roughness=LinearInterpolate(
    A=ScalarParameter() {group=Wrap_VeiledGlitch_A, parameterName=VGlit_A_Roughness_High}
    B=ScalarParameter() {defaultValue=1, group=Wrap_VeiledGlitch_A, parameterName=VGlit_A_Roughness_Low}
    Alpha=@n82.A
   )
   Specular=LinearInterpolate(
    A=ScalarParameter() {defaultValue=0.25, group=Wrap_VeiledGlitch_A, parameterName=VGlit_A_Spec}
    B=ScalarParameter() {defaultValue=0.25, group=Wrap_VeiledGlitch_A, parameterName=VGlit_A_Spec_Gadget}
    Alpha=TextureSample(UVs=@n81, Tex=@n83).G
   )
   Metallic=ScalarParameter() {group=Wrap_VeiledGlitch_A, parameterName=VGlit_A_Metalic}
   Normal=Add(A=@n80.Raw_Normal, B=NamedRerouteUsage())
   EmissiveColor
  ) {attributeSetTypes [
    #2
    #3
    #5
    #4
    #1
    #6
   ]}
  n85 = VectorParameter() {group=Wrap_VeiledGlitch_A_Accent, parameterName=VGlit_Accent_A_BaseColor, defaultValue=<rgba 0.0393 0.0291 0.05 1>}
  n86 = SetMaterialAttributes(
   MaterialAttributes=NamedRerouteUsage()
   BaseColor=@n85.RGB
   Roughness=ScalarParameter() {defaultValue=0.6, group=Wrap_VeiledGlitch_A_Accent, parameterName=VGlit_Accent_A_Roughness}
   Specular=ScalarParameter() {defaultValue=0.25, group=Wrap_VeiledGlitch_A_Accent, parameterName=VGlit_Accent_A_Spec}
   Metallic=ScalarParameter() {group=Wrap_VeiledGlitch_A_Accent, parameterName=VGlit_Accent_A_Metalic}
  ) {attributeSetTypes [
    #2
    #3
    #5
    #4
    #1
    #6
   ]}
  NamedRerouteDeclaration(None=BlendMaterialAttributes(A=@n84, B=@n86, Alpha=NamedRerouteUsage()) {vertexAttributeBlendType=UseA}) {name=Mat_Base_A, nodeColor=<rgba 0 0.3906 1 1>}
  n87 = SetMaterialAttributes(
   MaterialAttributes=NamedRerouteUsage()
   BaseColor=Constant()
   Roughness=ScalarParameter() {defaultValue=0.3, group=Wrap_VeiledGlitch_B, parameterName=VGlit_B_Roughness}
   Specular=ScalarParameter() {defaultValue=0.25, group=Wrap_VeiledGlitch_B, parameterName=VGlit_B_Spec}
   Metallic=ScalarParameter() {group=Wrap_VeiledGlitch_B, parameterName=VGlit_B_Metalic}
   Normal=@n9
   EmissiveColor=NamedRerouteUsage()
  ) {attributeSetTypes [
    #2
    #3
    #5
    #4
    #1
    #6
   ]}
  n88 = VectorParameter() {group=Wrap_VeiledGlitch_B_Accent, parameterName=VGlit_Accent_B_BaseColor, defaultValue=<rgba 0.013 0.0103 0.0296 1>}
  n89 = SetMaterialAttributes(
   MaterialAttributes=NamedRerouteUsage()
   BaseColor=@n88.RGB
   Roughness=ScalarParameter() {defaultValue=0.4, group=Wrap_VeiledGlitch_B_Accent, parameterName=VGlit_Accent_B_Roughness}
   Specular=ScalarParameter() {defaultValue=0.25, group=Wrap_VeiledGlitch_B_Accent, parameterName=VGlit_Accent_B_Spec}
   Metallic=ScalarParameter() {group=Wrap_VeiledGlitch_B_Accent, parameterName=VGlit_Accent_B_Metalic}
  ) {attributeSetTypes [
    #2
    #3
    #5
    #4
    #1
    #6
   ]}
  NamedRerouteDeclaration(None=BlendMaterialAttributes(A=@n87, B=@n89, Alpha=NamedRerouteUsage()) {vertexAttributeBlendType=UseA}) {name=Mat_Base_B, nodeColor=<rgba 0 0.3906 1 1>}
  NamedRerouteDeclaration(None=TextureObjectParameter() {group=Wrap_VeiledGlitch_A, parameterName=VGlit__Tex_BC, texture=&T_BlacksRGB}) {name=Tex_BC, nodeColor=<rgba 0.0781 0 1 1>}
  n90 = NamedRerouteUsage()
  n91 = NamedRerouteUsage()
  n92 = NamedRerouteUsage()
  n93 = NamedRerouteUsage()
  n94 = NamedRerouteUsage()
  n95 = NamedRerouteUsage()
  n96 = ScalarParameter() {defaultValue=0.8, group=Wrap_VeiledGlitch_B, parameterName=VGlit_B_Contrast}
  n97 = Constant() {r=1}
  n98 = ScalarParameter() {defaultValue=2, group=Wrap_VeiledGlitch_B, parameterName=VGlit_B_Scale_A_A}
  n99 = Multiply(A=Time(), B=VectorParameter() {group=Wrap_VeiledGlitch_B_UVS, parameterName=VGlit_B_Speed_A_A, defaultValue=<rgba 0.1 0.05 0 1>}.RGB)
  n100 = NamedRerouteUsage()
  n101 = NamedRerouteUsage()
  n102 = NamedRerouteUsage()
  n103 = NamedRerouteUsage()
  n104 = NamedRerouteUsage()
  n105 = NamedRerouteUsage()
  n106 = ScalarParameter() {defaultValue=1, group=Wrap_VeiledGlitch_B, parameterName=VGlit_B_Scale_A_B}
  n107 = Multiply(A=Time(), B=VectorParameter() {group=Wrap_VeiledGlitch_B_UVS, parameterName=VGlit_B_Speed_A_B, defaultValue=<rgba -0.05 -0.03 0 1>}.RGB)
  NamedRerouteDeclaration(None=Add(
    A=MaterialFunctionCall!(
     Texture_Normals=@n90
     Rotation_Axis=@n91
     Caculated_Bounds=@n92
     Pre_Skinned_Normals=@n93
     IsStatsB=@n94
     IsState=@n95
     Blend_Contrast=@n96
     Fill_alternate_Axis_Opacity=@n97
     Scale=@n98
     Offset_Pivot
     Offset=@n99
    ) {materialFunction=&/MF_TopsSideProjection_Normals_V2}.Add_Surface_Normal
    B=MaterialFunctionCall(
     Texture_Normals=@n100
     Rotation_Axis=@n101
     Caculated_Bounds=@n102
     Pre_Skinned_Normals=@n103
     IsStatsB=@n104
     IsState=@n105
     Blend_Contrast=@n96
     Fill_alternate_Axis_Opacity=@n97
     Scale=@n106
     Offset=@n107
    ) {materialFunction=&MF_TopsSideProjection_Normals_V2}.Add_Surface_Normal
   ) {constB=0}) {name=Normals, nodeColor=<rgba 0 1 0.6484 1>}
  n108 = NamedRerouteUsage()
  n109 = Multiply(
   A=Saturate(None=Add(A=@n108) {constB=-0.5})
   B=ScalarParameter() {defaultValue=8, group=Wrap_VeiledGlitch_B_Grid, parameterName=VGlit_B_Curvature_Strength}
  ) {constB=8}
  n110 = ScalarParameter() {defaultValue=5, group=Wrap_VeiledGlitch_B_Grid, parameterName=VGlit_B_Curvature_Falloff}
  n111 = Constant() {r=1}
  n112 = VectorParameter() {group=Wrap_VeiledGlitch_B_Grid, parameterName=VGlit_B_Axis_Color, defaultValue=<rgba 1 0 0 1>}
  n113 = MaterialFunctionCall(
   Texture=NamedRerouteUsage()
   Caculated_Bounds=NamedRerouteUsage()
   Pre_Skinned_Normals=NamedRerouteUsage()
   Blend_Contrast=@n111
   IsStatsB=NamedRerouteUsage()
   IsState=NamedRerouteUsage()
   Fill_alternate_Axis_Opacity=@n111
   "Fill_alternate_Axis_Color_A=Roughness"=AppendVector(A=@n112.RGB, B=@n112.RGB)
   Scale=NamedRerouteUsage()
   Rotation_Axis=NamedRerouteUsage()
  ) {materialFunction=&MF_TopsSideProjection_Base_V2}
  n114 = Add(A=Saturate(None=Power(Base=@n109, Exp=@n110) {constExponent=5}), B=ComponentMask(None=Power(Base=@n113.Combined)) {r=true})
  NamedRerouteDeclaration(None=Multiply(A=Add(A=@n114, B=@n6), B=NamedRerouteUsage())) {name=Grid_Pattern, nodeColor=<rgba 0 1 0.2031 1>}
  NamedRerouteDeclaration(None=@n6) {name=Grad_Msk, nodeColor=<rgba 0 1 0.1562 1>}
  NamedRerouteDeclaration(None=@n80.UVs No Offsets/Tiling) {name=Uvs_2, nodeColor=<rgba 0 0.3672 1 1>}
  n115 = CurveAtlasRowParameter(CurveTime=@n65.A) {
   atlas=&CruveAtlas_Wraps_01
   curve=/Game/Weapons/WeaponSkins/Wraps/Textures/V_Glitch/CC_VGlitch_Bc2
   defaultValue=103
   group=Wrap_VeiledGlitch_B_Noise
   parameterName=VGlit__CC_NoiseBC
  }
  n116 = MaterialFunctionCall(
   Index=ScalarParameter() {defaultValue=103, group=Wrap_VeiledGlitch_B_Noise, parameterName="VGlit__CC_NoiseBC  Index"}
   CCinput=@n115
   CCinput_Alpha=@n115
   UV=@n65.A
  ) {materialFunction=&MF_Global_CC_Tex}
  n117 = Multiply(A=NamedRerouteUsage(), B=Multiply(A=@n65.A, B=@n116.3V))
  NamedRerouteDeclaration(None=Multiply(A=OneMinus(None=@n60), B=Multiply(A=@n117, B=@n116.Alpha))) {name=Glitch_Noise, nodeColor=<rgba 1 0 0.2344 1>}
  NamedRerouteDeclaration(None=OneMinus(None=ComponentMask(None=@n113.Combined) {r=true})) {name=Alt_Msk, nodeColor=<rgba 1 0.3516 0 1>}
  n118 = ScalarParameter() {defaultValue=102, group=Wrap_VeiledGlitch_Reactive, parameterName="VGlit__CC_Grad  Index"}
  n119 = Frac(None=@n27)
  n120 = CurveAtlasRowParameter(CurveTime=@n119) {
   atlas=&CruveAtlas_Wraps_01
   curve=&CC_VGlitch_Msk
   defaultValue=102
   group=Wrap_VeiledGlitch_B_Grid
   parameterName=VGlit__CC_Grad
  }
  NamedRerouteDeclaration(None=MaterialFunctionCall(Index=@n118, CCinput_Alpha=@n120.A, UV=@n119) {materialFunction=&MF_Global_CC_Tex}.Alpha) {name=Brightglow, nodeColor=<rgba 1 0.1406 0 1>}
  n121 = CurveAtlasRowParameter(CurveTime=@n29) {
   atlas=&CruveAtlas_Wraps_01
   curve=&CC_VGlitch_Msk
   defaultValue=102
   group=Wrap_VeiledGlitch_B_Grid
   parameterName=VGlit__CC_Grad
  }
  n122 = ComponentMask(None=MaterialFunctionCall(Index=@n118, CCinput=@n121, UV=@n29) {materialFunction=&MF_Global_CC_Tex}.3V) {b=true}
  NamedRerouteDeclaration(None=@n122) {name=Pixlation mask, nodeColor=<rgba 1 0.5156 0 1>}
  n123 = CurveAtlasRowParameter(CurveTime=@n27) {
   atlas=&CruveAtlas_Wraps_01
   curve=&CC_VGlitch_Msk
   defaultValue=102
   group=Wrap_VeiledGlitch_B_Grid
   parameterName=VGlit__CC_Grad
  }
  NamedRerouteDeclaration(None=MaterialFunctionCall(Index=@n118, CCinput_Alpha=@n123.A) {materialFunction=&MF_Global_CC_Tex}.Alpha) {name=Reactive_Mask, nodeColor=<rgba 0 1 0.0156 1>}
  NamedRerouteDeclaration(None=@n34) {name=Msk_Accent, nodeColor=<rgba 0 1 0.7656 1>}
  n124 = NamedRerouteUsage()
  n125 = OneMinus(None=@n124)
  n126 = ScalarParameter() {group=Wrap_VeiledGlitch_Reactive, parameterName=VGlit_Reactive_Msk_invert}
  n127 = GetMaterialAttributes(None=BlendMaterialAttributes(A=NamedRerouteUsage(), B=NamedRerouteUsage(), Alpha=LinearInterpolate(A=@n124, B=@n125, Alpha=@n126)) {vertexAttributeBlendType=UseA}) {attributeGetTypes [#2, #6]}
  n128 = VectorParameter() {group=Wrap_VeiledGlitch_Reactive, parameterName=VGlit_Reactive_Color_B, defaultValue=<rgba 0 0.6063 4 1>}
  n129 = VectorParameter() {group=Wrap_VeiledGlitch_Reactive, parameterName=VGlit_Reactive_Color_A, defaultValue=<rgba 0.8282 0.7988 1 1>}
  n130 = NamedRerouteUsage()
  n131 = Multiply(A=@n130, B=LinearInterpolate(A=@n128.RGB, B=@n129.RGB, Alpha=@n130))
  NamedRerouteDeclaration(None=SetMaterialAttributes!(MaterialAttributes=@n127.MaterialAttributes, BaseColor=@n127.MaterialAttributes, EmissiveColor=Add(A=@n127.EmissiveColor, B=@n131)) {attributeSetTypes [#2, #6]}) {name=Final_output, nodeColor=<rgba 0 0.8125 1 1>}
  n132 = Subtract(A=Time(), B=ScalarParameter() {defaultValue=-9999999, group=Wrap_VeiledGlitch_Reactive, parameterName=VGlit_Reactive_Time_BP})
  n133 = Saturate(None=Divide(A=@n132, B=ScalarParameter() {defaultValue=0.25, group=Wrap_VeiledGlitch_Reactive, parameterName=VGlit_Reactive_Time_BP_Duration}))
  NamedRerouteDeclaration(None=@n133) {name=Reactive, nodeColor=<rgba 0.1016 0 1 1>}
  NamedRerouteDeclaration(None=@n18.PS) {name=Preskin_Pos, nodeColor=<rgba 0 0.9297 1 1>}
  NamedRerouteDeclaration(None=TextureObjectParameter() {group=Wrap_VeiledGlitch_B, parameterName=VGlit__Tex_Msk2, texture=&T_BlacksRGB}) {name=Tex_Msk2, nodeColor=<rgba 0.0781 0 1 1>}
  n134 = Add(A=@n39, B=NamedRerouteUsage())
  n135 = ScalarParameter() {defaultValue=2, group=Wrap_VeiledGlitch_B, parameterName=VGlit_B_Msk_Falloff}
  NamedRerouteDeclaration(None=Clamp(
    None=Divide(A=@n134, B=@n135)
    Min=ScalarParameter() {defaultValue=0.15, group=Wrap_VeiledGlitch_B, parameterName=VGlit_B_Msk_Bottom Color}
    Max=ScalarParameter() {defaultValue=0.95, group=Wrap_VeiledGlitch_B, parameterName=VGlit_B_Msk_Top Color}
   )) {name=Top to Bottom gradient, nodeColor=<rgba 0.2422 0 1 1>}
  NamedRerouteDeclaration(None=@n114) {name=Msk_Grid, nodeColor=<rgba 1 0.3516 0 1>}
  NamedRerouteDeclaration(None=@n13.3V) {name=B_BaseColor, nodeColor=<rgba 1 0 0.5156 1>}
  n136 = Divide(A=@n132, B=ScalarParameter() {defaultValue=0.25, group=Wrap_VeiledGlitch_Reactive, parameterName=VGlit_Reactive_Time_BP_Duration_Glow})
  NamedRerouteDeclaration(None=Saturate(None=Add(B=@n136) {constA=0.1})) {name=Reactive_1, nodeColor=<rgba 0.1016 0 1 1>}
  n137 = Multiply(A=NamedRerouteUsage(), B=ScalarParameter() {defaultValue=0.45, group=Wrap_VeiledGlitch_B_Noise, parameterName=VGlit_B_Base Emissive})
  n138 = NamedRerouteUsage()
  n139 = Multiply(A=@n45.RGB, B=ScalarParameter() {defaultValue=55, group=Wrap_VeiledGlitch_B_Noise, parameterName=VGlit_B_Noise_B_Warp_Amount})
  n140 = ScalarParameter() {defaultValue=200, group=Wrap_VeiledGlitch_B_Noise, parameterName=VGlit_B_Noise_A_Size}
  n141 = MaterialFunctionCall(Float4=NamedRerouteUsage()) {materialFunction=&BreakOutFloat4Components}
  n142 = LinearInterpolate(
   A=@n140
   B=LinearInterpolate(
    A=ScalarParameter() {defaultValue=200, group=Wrap_VeiledGlitch_B_Noise, parameterName=VGlit_B_Noise_A_Size Vehicle}
    B=ScalarParameter() {defaultValue=200, group=Wrap_VeiledGlitch_B_Noise, parameterName=VGlit_B_Noise_A_Size Vehicle Side}
    Alpha=@n141.G
   )
   Alpha=@n141.R
  )
  n143 = LinearInterpolate(A=ScalarParameter() {defaultValue=200, group=Wrap_VeiledGlitch_B_Noise, parameterName=VGlit_B_Noise_A_Size Mech}, B=@n140, Alpha=@n141.A)
  n144 = VectorParameter() {group=Wrap_VeiledGlitch_B_Noise, parameterName=VGlit_B_Noise_A_Speed, defaultValue=<rgba -0.15 0.05 0 1>}
  n145 = MaterialFunctionCall(
   Texture=NamedRerouteUsage()
   Caculated_Bounds=Divide(A=Add(A=@n138, B=@n139), B=LinearInterpolate(A=@n142, B=@n143, Alpha=@n141.B))
   Pre_Skinned_Normals=NamedRerouteUsage()
   Blend_Contrast=Constant()
   IsStatsB=NamedRerouteUsage()
   IsState=NamedRerouteUsage()
   Fill_alternate_Axis_Opacity=Constant()
   "Fill_alternate_Axis_Color_A=Roughness"=Constant4Vector().RGBA
   Offset=Multiply(A=Time(), B=@n144.RGB)
   Rotation_Axis=NamedRerouteUsage()
  ) {materialFunction=&MF_TopsSideProjection_Base_V2}
  n146 = ComponentMask(None=@n45.RGB) {g=true}
  n147 = Multiply(A=ComponentMask(None=@n145.RGB) {r=true}, B=Add(A=@n146) {constB=0.5})
  n148 = ComponentMask(None=@n45.RGB) {b=true}
  n149 = Power(Base=@n148) {constExponent=4}
  n150 = Multiply(
   A=ScalarParameter() {defaultValue=1, group=Wrap_VeiledGlitch_B_Noise, parameterName=VGlit_B_Warp Freq}
   B=Power(Base=Subtract(A=@n147, B=@n149)) {constExponent=3}
  )
  n151 = ComponentMask(None=Saturate(None=Sine(None=@n150))) {r=true}
  n152 = CurveAtlasRowParameter(CurveTime=@n151) {
   atlas=&CruveAtlas_Wraps_01
   curve=/Game/Characters/Player/Female/Medium/Bodies/F_MED_Veiled_Glitch/Materials/VeiledGlitchGradient
   defaultValue=104
   group=Wrap_VeiledGlitch_B_Noise
   parameterName=VGlit_B_CColors
  }
  n153 = MaterialFunctionCall(Index=ScalarParameter() {defaultValue=104, group=Wrap_VeiledGlitch_B_Noise, parameterName="VGlit_B_CColors  Index"}, CCinput=@n152, UV=@n151) {materialFunction=&MF_Global_CC_Tex}
  n154 = Multiply(A=ComponentMask(None=@n145.RGB) {g=true}, B=Add(A=@n148) {constB=0.5})
  n155 = ScalarParameter() {defaultValue=0.8, group=Wrap_VeiledGlitch_B_Noise, parameterName=VGlit_B_Freq bias}
  n156 = Saturate(None=Power(Base=Multiply(A=@n154, B=@n155), Exp=ScalarParameter() {defaultValue=4, group=Wrap_VeiledGlitch_B_Noise, parameterName=VGlit_B_Freq}))
  n157 = Fresnel() {baseReflectFraction=0, exponent=1}
  n158 = OneMinus(None=Power(Base=@n157, Exp=ScalarParameter() {defaultValue=0.25, group=Wrap_VeiledGlitch_B_Noise, parameterName=VGlit_B_Fresnel Power}))
  n159 = Multiply(
   A=LinearInterpolate(
    A=ScalarParameter() {defaultValue=20, group=Wrap_VeiledGlitch_B, parameterName=VGlit_B_Emissive Intensity Bottom}
    B=ScalarParameter() {defaultValue=5, group=Wrap_VeiledGlitch_B, parameterName=VGlit_B_Emissive Intensity Top}
    Alpha=NamedRerouteUsage()
   )
   B=Multiply(A=LinearInterpolate(A=@n137, B=@n153.3V, Alpha=@n156), B=Clamp(None=@n158) {maxDefault=99, minDefault=0.05})
  )
  n160 = Add(
   A=@n159
   B=Multiply(
    A=VectorParameter() {group=Wrap_VeiledGlitch_B_Grid, parameterName=Gridline Color, defaultValue=<rgba 5 1.3125 3.8628 1>}.RGB
    B=NamedRerouteUsage()
   )
  )
  n161 = MaterialFunctionCall(
   In=Multiply(
    A=ComponentMask(None=@n45.RGB) {b=true}
    B=ScalarParameter() {defaultValue=1, group=Wrap_VeiledGlitch_B_Grid, parameterName=VGlit_B_Gridline Frequency}
   )
   Contrast=ScalarParameter() {defaultValue=3, group=Wrap_VeiledGlitch_B_Grid, parameterName=VGlit_B_Gridline Contrast}
  ) {materialFunction=&CheapContrast}
  n162 = ScalarParameter() {defaultValue=1, group=Wrap_VeiledGlitch_B_Grid, parameterName=VGlit_B_Gridline Boost}
  n163 = Power(Base=@n157, Exp=ScalarParameter() {defaultValue=0.2, group=Wrap_VeiledGlitch_B_Noise, parameterName=VGlit_B_Fresnel Power})
  n164 = Multiply(A=Saturate(None=Multiply(A=@n161.Result, B=@n162)), B=Saturate(None=OneMinus(None=@n163)))
  n165 = Constant() {r=0.5}
  n166 = Constant() {r=1}
  NamedRerouteDeclaration(None=Multiply(
    A=LinearInterpolate(A=@n159, B=@n160, Alpha=@n164)
    B=MaterialFunctionCall(Value_Morning=@n165, Value_Day=@n165, Value_Evening=@n165, Value_Night=@n166) {materialFunction=&Scalar_Time_of_Day}.Result
   )) {name=B_Material_Emissive, nodeColor=<rgba 0.5234 1 0 1>}
  n167 = MaterialFunctionCall(Color=NamedRerouteUsage()) {materialFunction=&SplitComponents}
  n168 = MaterialFunctionCall(Float4=NamedRerouteUsage()) {materialFunction=&BreakOutFloat4Components}
  n169 = LinearInterpolate(
   A=LinearInterpolate(A=@n167.G, B=LinearInterpolate(A=@n167.R, B=@n167.R, Alpha=@n168.G), Alpha=@n168.R)
   B=LinearInterpolate(A=@n167.B, B=@n167.B, Alpha=@n168.A)
   Alpha=@n168.B
  )
  n170 = MaterialFunctionCall(Float2=NamedRerouteUsage()) {materialFunction=&BreakOutFloat2Components}
  NamedRerouteDeclaration(None=LinearInterpolate(A=LinearInterpolate(A=@n169, B=@n167.B, Alpha=@n170.R), B=@n167.G, Alpha=@n170.G)) {name=Masking Direction, nodeColor=<rgba 0 0.625 1 1>}
  NamedRerouteDeclaration(None=FunctionInput(Preview=Constant4Vector().RGBA) {bUsePreviewValueAsDefault=true, inputName=IsState, inputType=FunctionInput_Vector4, sortPriority=4}) {name=IsState, nodeColor=<rgba 0.1497 1 0.1633 1>}
  n171 = ScalarParameter() {defaultValue=3, group=Wrap_VeiledGlitch_A_UVS, parameterName=VGlit_B_Grid_ScaleVehicle, sortPriority=150}
  n172 = ScalarParameter() {defaultValue=3, group=Wrap_VeiledGlitch_A_UVS, parameterName=VGlit_B_Grid_ScaleVehicle_Side, sortPriority=150}
  n173 = MaterialFunctionCall(Float4=NamedRerouteUsage()) {materialFunction=&BreakOutFloat4Components}
  n174 = LinearInterpolate(
   A=LinearInterpolate(
    A=ScalarParameter() {defaultValue=3, group=Wrap_VeiledGlitch_A_UVS, parameterName=VGlit_B_Grid_ScaleWeapon, sortPriority=150}
    B=LinearInterpolate(A=@n171, B=@n172, Alpha=@n173.G)
    Alpha=@n173.R
   )
   B=LinearInterpolate(
    A=ScalarParameter() {defaultValue=3, group=Wrap_VeiledGlitch_A_UVS, parameterName=VGlit_B_Grid_ScaleMech, sortPriority=150}
    B=ScalarParameter() {defaultValue=3, group=Wrap_VeiledGlitch_A_UVS, parameterName=VGlit_B_Grid_ScaleCharacter, sortPriority=150}
    Alpha=@n173.A
   )
   Alpha=@n173.B
  )
  n175 = ScalarParameter() {defaultValue=3, group=Wrap_VeiledGlitch_A_UVS, parameterName=VGlit_B_Grid_ScaleBackpack, sortPriority=150}
  n176 = MaterialFunctionCall(Float2=NamedRerouteUsage()) {materialFunction=&BreakOutFloat2Components}
  NamedRerouteDeclaration(None=LinearInterpolate(
    A=LinearInterpolate(A=@n174, B=@n175, Alpha=@n176.R)
    B=ScalarParameter() {defaultValue=3, group=Wrap_VeiledGlitch_A_UVS, parameterName=VGlit_B_Grid_ScalePickaxe, sortPriority=150}
    Alpha=@n176.G
   )) {name=UVs_Scale_Grid, nodeColor=<rgba 0 0.7188 1 1>}
  n177 = ScalarParameter() {group=Wrap_VeiledGlitch_A_UVS, parameterName=VGlit_B_Msk_Offset_Vehicle, sortPriority=150}
  n178 = ScalarParameter() {group=Wrap_VeiledGlitch_A_UVS, parameterName=VGlit_B_Msk_Offset_Vehicle_Side, sortPriority=150}
  n179 = MaterialFunctionCall(Float4=NamedRerouteUsage()) {materialFunction=&BreakOutFloat4Components}
  n180 = LinearInterpolate(
   A=LinearInterpolate(
    A=ScalarParameter() {group=Wrap_VeiledGlitch_A_UVS, parameterName=VGlit_B_Msk_Offset_Weapon, sortPriority=150}
    B=LinearInterpolate(A=@n177, B=@n178, Alpha=@n179.G)
    Alpha=@n179.R
   )
   B=LinearInterpolate(
    A=ScalarParameter() {group=Wrap_VeiledGlitch_A_UVS, parameterName=VGlit_B_Msk_Offset_Mech, sortPriority=150}
    B=ScalarParameter() {group=Wrap_VeiledGlitch_A_UVS, parameterName=VGlit_B_Msk_Offset_Character, sortPriority=150}
    Alpha=@n179.A
   )
   Alpha=@n179.B
  )
  n181 = ScalarParameter() {group=Wrap_VeiledGlitch_A_UVS, parameterName=VGlit_B_Msk_Offset_Backpack, sortPriority=150}
  n182 = MaterialFunctionCall(Float2=NamedRerouteUsage()) {materialFunction=&BreakOutFloat2Components}
  NamedRerouteDeclaration(None=LinearInterpolate(
    A=LinearInterpolate(A=@n180, B=@n181, Alpha=@n182.R)
    B=ScalarParameter() {group=Wrap_VeiledGlitch_A_UVS, parameterName=VGlit_B_Msk_Offset_Pickaxe, sortPriority=150}
    Alpha=@n182.G
   )) {name=UVs_Msk_Offset, nodeColor=<rgba 0 0.7188 1 1>}
  NamedRerouteDeclaration(None=MaterialFunctionCall(UVs=@n80.UVs No Offsets/Tiling, Rotation Center=Constant() {r=0.5}, "Rotation Angle (0-1)"=@n79) {materialFunction=&CustomRotator}.Rotated Values) {name=Uvs_3, nodeColor=<rgba 0 0.3672 1 1>}
  NamedRerouteDeclaration(None=FunctionInput(Preview=Constant2Vector()) {bUsePreviewValueAsDefault=true, inputName=IsState_B, inputType=FunctionInput_Vector2, sortPriority=4}) {name=IsState_B, nodeColor=<rgba 0.1497 1 0.1633 1>}
  NamedRerouteDeclaration(None=@n22) {name=Caclcated_Bounds, nodeColor=<rgba 0 0.9297 1 1>}
  n183 = TextureObjectParameter() {
   group=Wrap_VeiledGlitch_A
   parameterName=VGlit__Tex_N
   samplerType=SAMPLERTYPE_Normal
   texture=&FlatNormalBlue
  }
  NamedRerouteDeclaration(None=@n183) {name=Tex_N, nodeColor=<rgba 0.0781 0 1 1>}
  n184 = PreSkinnedNormal()
  NamedRerouteDeclaration(None=FunctionInput(Preview=VertexInterpolator(VS=@n184).PS) {bUsePreviewValueAsDefault=true, inputName=Pre-Skinned Local Normal, sortPriority=10}) {name=PreSkin_Normal, nodeColor=<rgba 0.8281 1 0 1>}
  NamedRerouteDeclaration(None=@n38.Normal) {name=Root_Normal, nodeColor=<rgba 0.4062 0 1 1>}
  NamedRerouteUsage()
  NamedRerouteUsage()
  NamedRerouteUsage()
  NamedRerouteUsage()
  NamedRerouteUsage()
  NamedRerouteUsage()
  NamedRerouteUsage()
  NamedRerouteUsage()
  OneMinus(None=@n7)
  OneMinus(None=@n133)
  OneMinus(None=@n7)
  Reroute(None=@n122)
  n185 = Multiply(A=Time()) {constB=0.25}
  n186 = ScalarParameter() {defaultValue=1.5, parameterName=Debug param}
  Saturate(None=Divide(A=Fmod(A=@n185, B=@n186), B=@n186))
  ScalarParameter() {defaultValue=2.5, group=Wrap_VeiledGlitch_B_Grid, parameterName=VGlit_B_Grid_Scale}
  ScalarParameter() {defaultValue=0.25, group=Wrap_VeiledGlitch_A, parameterName=VGlit_A_Spec}
  ScalarParameter() {group=Wrap_VeiledGlitch_Reactive, parameterName=VGlit_Reactive_Msk_End}
  ScalarParameter() {defaultValue=1, group=Wrap_VeiledGlitch_Reactive, parameterName=VGlit_Reactive_Msk_Start}
  ScalarParameter() {defaultValue=200, group=Wrap_VeiledGlitch_B_Noise, parameterName=VGlit_B_Noise_B_Size_00}
  ScalarParameter() {defaultValue=200, group=Wrap_VeiledGlitch_B_Noise, parameterName=VGlit_B_Noise_A_Size_00}
  SetMaterialAttributes!(MaterialAttributes=NamedRerouteUsage(), BaseColor=Constant(), Roughness, Specular, Metallic, Normal, EmissiveColor=NamedRerouteUsage()) {attributeSetTypes [
    #2
    #3
    #5
    #4
    #1
    #6
   ]}
  Subtract(A=ComponentMask(None=@n145.RGB) {g=true}, B=Divide(A=@n146) {constB=3})
  TextureSample(UVs=@n26.UVs) {texture=/Game/Weapons/WeaponSkins/Wraps/Textures/UVCheaker_Clamped}
 ]}
