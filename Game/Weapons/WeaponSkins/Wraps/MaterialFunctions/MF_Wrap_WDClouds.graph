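# Material function for the WomanDay01 item wrap: an animated emissive night sky of clouds, moon and stars, toggled by a static switch.
# Side, safe-zone and cheap triplanar UVs place the layers, tuned per weapon, vehicle, mech, character, backpack and pickaxe.
# Panned noise warps clouds, grid cells randomize star spin and twinkle, and radial gradients drive moon glow.
# Primary and secondary masks blend base and accent materials, with time-of-day emissive scaling.
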MaterialFunction /Game/Weapons/WeaponSkins/Wraps/MaterialFunctions/MF_Wrap_WDClouds
graph {nodes [
  n1 = NamedRerouteUsage()
  n2 = Multiply(A=Time(), B=ScalarParameter() {defaultValue=-0.5, group=Wrap_WomanDay01_Clouds__FRG, parameterName=WD01_CloudBKG_Master_Speed})
  n3 = NamedRerouteUsage()
  n4 = TextureSample(UVs=Panner(Coordinate=Multiply(A=@n1) {constB=3}, Time=@n2) {speedX=0.025, speedY=0.002}, Tex=@n3) {samplerSource=SSM_Wrap_WorldGroupSettings}
  n5 = TextureSample(UVs=Panner(Coordinate=Multiply(A=@n1) {constB=2}, Time=@n2) {speedX=-0.05, speedY=0.002}, Tex=@n3) {samplerSource=SSM_Wrap_WorldGroupSettings}
  n6 = Add(A=@n4.R, B=@n5.G)
  n7 = Multiply(A=@n4.G, B=@n5.R)
  n8 = Multiply(A=AppendVector(A=Multiply(A=@n6, B=@n7), B=@n7)) {constB=0.025}
  n9 = NamedRerouteUsage()
  n10 = Panner(Coordinate=Multiply(A=@n9, B=Constant2Vector() {g=1.5, r=1.5}), Time=@n2) {speedX=-0.012, speedY=0.01}
  Add(A=@n8, B=@n10)
  n11 = Panner(Coordinate=Multiply(A=@n9, B=Constant2Vector() {g=2.5, r=2.5}), Time=@n2) {speedX=-0.01, speedY=-0.01}
  Add(A=@n11, B=@n8)
  n12 = MaterialFunctionCall(Float4=NamedRerouteUsage()) {materialFunction=&BreakOutFloat4Components}
  n13 = MaterialFunctionCall!(
   For_Safe_Zones
   Normal_Map
   Caculated_Bounds=NamedRerouteUsage()
   Pre_Skinned_Normals
   Is Vehicle=@n12.R
   Is VehicleSide=@n12.R
   Is Mech=@n12.R
   Vehicle_Rotation
   UV_Rotation Bypass
   Mech_Rotation
   Uvs_Offset
   Offset-after rotation
   Uvs_TileRate
  ) {materialFunction=&/MF_Projection_Side_Normal}
  n14 = MaterialFunctionCall(
   UVs=Add(A=NamedRerouteDeclaration(None=@n13.UVs No Offsets/Tiling) {name=Uvs_Side_projection, nodeColor=<rgba 0.6641 1 0 1>}, B=NamedRerouteUsage())
   Rotation Center=Constant2Vector() {g=0.5, r=0.5}
   "Rotation Angle (0-1)"=NamedRerouteUsage()
  ) {materialFunction=&CustomRotator}
  Add(A=@n14.Rotated Values) {constB=0}
  n15 = ScalarParameter() {defaultValue=20, group=Wrap_WomanDay01_Stars__FRG, parameterName=WD01_Stars_TileRate}
  n16 = Time()
  n17 = Constant2Vector()
  n18 = NamedRerouteUsage()
  n19 = Add(A=@n18, B=ComponentMask(None=Multiply(A=@n16, B=@n17)) {g=true, r=true}) {constB=0}
  n20 = ComponentMask(None=@n19) {r=true}
  n21 = Divide(A=Floor(None=Multiply(A=@n20, B=@n15) {constB=5}), B=@n15) {constB=5}
  n22 = Add(A=Multiply(A=@n19, B=@n15), B=AppendVector(A=Constant(), B=Multiply(A=@n21) {constB=3})) {constB=0}
  n23 = Frac(None=@n22)
  Add(A=@n23, B=LinearInterpolate(A=Constant2Vector() {g=-0.132, r=0.18}, B=Constant2Vector() {g=0.18, r=-0.13}) {constB=0})
  n24 = NamedRerouteUsage()
  n25 = NamedRerouteUsage()
  n26 = VectorParameter() {group=Wrap_WomanDay01_Moon, parameterName=WD01_Moon_Color_AC, defaultValue=<rgba 0.099 0.0697 0.2083 1>}
  n27 = VectorParameter() {group=Wrap_WomanDay01_Moon, parameterName=WD01_Moon_Color_AB, defaultValue=<rgba 0.1356 0.1832 0.5208 1>}
  n28 = NamedRerouteUsage()
  n29 = ComponentMask(None=@n28) {g=true}
  n30 = Add(A=ComponentMask(None=@n28) {r=true}, B=Multiply(A=OneMinus(None=@n29)) {constB=0.5})
  n31 = Saturate(None=Power(Base=Multiply(A=@n30) {constB=0.5}) {constExponent=2.5})
  n32 = TextureSample(UVs=@n28, Tex=NamedRerouteUsage())
  n33 = LinearInterpolate(
   A=@n26.RGB
   B=LinearInterpolate(
    A=@n27.RGB
    B=VectorParameter() {group=Wrap_WomanDay01_Moon, parameterName=WD01_Moon_Color_AA, defaultValue=<rgba 2.0942 1.4687 3 1>}.RGB
    Alpha=@n31
   )
   Alpha=@n32.RGB
  )
  n34 = NamedRerouteDeclaration(None=Multiply(A=@n32.A, B=NamedRerouteUsage())) {name=Moon_Msk, nodeColor=<rgba 1 0.0703 0 1>}
  n35 = VectorParameter() {group=Wrap_WomanDay01_Moon, parameterName=WD01_MoonGlow_ColorFRG, defaultValue=<rgba 2.5189 1.8166 3.5 1>}
  n36 = VectorParameter() {group=Wrap_WomanDay01_Clouds__FRG, parameterName=WD01_Cloud_Color_BB, defaultValue=<rgba 0.0236 0.0355 0.1302 1>}
  n37 = VectorParameter() {group=Wrap_WomanDay01_Clouds__FRG, parameterName=WD01_Cloud_Color_AB, defaultValue=<rgba 0.0802 0.1209 0.4427 1>}
  n38 = MaterialFunctionCall(UVs=NamedRerouteUsage(), Rotation Center=Constant2Vector() {g=0.5, r=0.5}, "Rotation Angle (0-1)"=Constant() {r=0.5}) {materialFunction=&CustomRotator}
  n39 = Constant2Vector() {g=2.5, r=2.5}
  n40 = Multiply(A=Time(), B=ScalarParameter() {defaultValue=1, group=Wrap_WomanDay01_Clouds__FRG, parameterName=WD01_Cloud_Master_Speed})
  n41 = NamedRerouteUsage()
  n42 = NamedRerouteUsage()
  n43 = TextureSample(UVs=Panner(Coordinate=Multiply(A=@n41) {constB=3}, Time=@n40) {speedX=0.015, speedY=0.002}, Tex=@n42) {samplerSource=SSM_Wrap_WorldGroupSettings}
  n44 = NamedRerouteUsage()
  n45 = TextureSample(UVs=Panner(Coordinate=Multiply(A=@n44) {constB=2}, Time=@n40) {speedX=-0.03, speedY=0.002}, Tex=@n42) {samplerSource=SSM_Wrap_WorldGroupSettings}
  n46 = Add(A=@n43.R, B=@n45.G)
  n47 = Multiply(A=@n43.G, B=@n45.R)
  n48 = Multiply(A=AppendVector(A=Multiply(A=@n46, B=@n47), B=@n47)) {constB=0.025}
  n49 = Add(A=Panner(Coordinate=Multiply(A=@n38.Rotated Values, B=@n39), Time=@n40) {speedX=-0.01, speedY=-0.005}, B=@n48)
  n50 = NamedRerouteUsage()
  n51 = Multiply(A=@n38.Rotated Values, B=Constant2Vector() {g=1.5, r=1.5})
  n52 = TextureSample(UVs=Add(A=@n48, B=Panner(Coordinate=@n51, Time=@n40) {speedX=-0.012, speedY=0.005}), Tex=@n50) {samplerSource=SSM_Wrap_WorldGroupSettings, samplerType=SAMPLERTYPE_Masks}
  n53 = Multiply(A=Add(A=TextureSample(UVs=@n49, Tex=@n50) {samplerSource=SSM_Wrap_WorldGroupSettings, samplerType=SAMPLERTYPE_Masks}.R, B=@n52.G), B=@n52.R)
  n54 = NamedRerouteUsage()
  n55 = ScalarParameter() {group=Wrap_WomanDay01_Clouds__FRG, parameterName=WD01_Clouds_FRG_Offset}
  n56 = Subtract(A=MaterialFunctionCall(
    UVs=Add(A=@n54, B=@n55) {constB=0}
    Texture Scale=ScalarParameter() {defaultValue=3, group=Wrap_WomanDay01_Clouds__FRG, parameterName=WD01_Clouds_FRG_Falloff_Scale}
   ) {materialFunction=&ScaleUVsByCenter}.UVs) {constB=0.5}
  n57 = LinearInterpolate(
   A=ScalarParameter() {defaultValue=1.1, group=Wrap_WomanDay01_Clouds__FRG, parameterName=WD01_Clouds_FRG_Falloff_Whitepoint}
   Alpha=ComponentMask(None=Abs(None=@n56)) {g=true}
  ) {constA=1.1, constB=-1}
  n58 = NamedRerouteUsage()
  n59 = Multiply(
   A=Power(Base=Saturate(None=@n57), Exp=ScalarParameter() {defaultValue=5, group=Wrap_WomanDay01_Clouds__FRG, parameterName=WD01_Clouds_FRG_Falloff})
   B=TextureSample(UVs=Panner(Coordinate=@n58) {speedX=-0.018, speedY=0.005}, Tex=NamedRerouteUsage()) {samplerType=SAMPLERTYPE_Masks}.A
  )
  n60 = Multiply(A=Power(Base=LinearInterpolate(Alpha=@n53) {constB=0.6}) {constExponent=1.5}, B=Saturate(None=Add(A=@n59) {constB=0.25}))
  n61 = VectorParameter() {group=Wrap_WomanDay01_Clouds__FRG, parameterName=WD01_Cloud_Color_BA, defaultValue=<rgba 0.0785 0.0211 0.2135 1>}
  n62 = VectorParameter() {group=Wrap_WomanDay01_Clouds__FRG, parameterName=WD01_Cloud_Color_AA, defaultValue=<rgba 0.2654 0.1408 0.6823 1>}
  n63 = NamedRerouteUsage()
  n64 = ComponentMask(None=Abs(None=Add(A=@n63) {constB=-0.5})) {r=true}
  n65 = Divide(A=@n64, B=ScalarParameter() {defaultValue=0.5, group=Wrap_WomanDay01_Clouds__FRG, parameterName=WD01_Cloud_Grad_Falloff}) {constB=0.5}
  n66 = VectorParameter() {group=Wrap_WomanDay01_Clouds__FRG, parameterName=WD01_Cloud_Color_BC, defaultValue=<rgba 0.0625 0.0388 0.0194 1>}
  n67 = VectorParameter() {group=Wrap_WomanDay01_Clouds__FRG, parameterName=WD01_Cloud_Color_AC, defaultValue=<rgba 0.6823 0.5004 0.2246 1>}
  n68 = ScalarParameter() {defaultValue=0.5, group=Wrap_WomanDay01_Clouds__FRG, parameterName=WD01_Cloud_Grad_FalloffB}
  n69 = OneMinus(None=Saturate(None=Divide(A=@n64, B=@n68) {constB=0.5}))
  n70 = TextureSample(UVs=Panner(Coordinate=NamedRerouteUsage(), Time=@n40) {speedX=-0.025, speedY=-0.002}, Tex=@n50) {samplerSource=SSM_Wrap_WorldGroupSettings, samplerType=SAMPLERTYPE_Masks}
  n71 = LinearInterpolate(
   A=LinearInterpolate(A=LinearInterpolate(A=@n36.RGB, B=@n37.RGB, Alpha=@n60), B=LinearInterpolate(A=@n61.RGB, B=@n62.RGB, Alpha=@n60), Alpha=Saturate(None=@n65))
   B=LinearInterpolate(A=@n66.RGB, B=@n67.RGB, Alpha=@n60)
   Alpha=Multiply(
    A=Multiply(A=@n69, B=@n70.B)
    B=ScalarParameter() {defaultValue=0.1, group=Wrap_WomanDay01_Clouds__FRG, parameterName=WD01_Cloud_third_Color}
   ) {constB=0}
  )
  n72 = Multiply(A=@n71, B=@n35.RGB)
  n73 = NamedRerouteUsage()
  n74 = ScalarParameter() {defaultValue=0.2, group=Wrap_WomanDay01_Moon, parameterName=WD01_Glow_FRGMoon_Msk}
  n75 = Saturate(None=Add(A=NamedRerouteUsage(), B=Multiply(A=@n73, B=@n74) {constB=0.1}))
  n76 = ScalarParameter() {defaultValue=1, group=Wrap_WomanDay01_Clouds__FRG, parameterName=WD01_Clouds_FRG_Densty}
  n77 = ScalarParameter() {defaultValue=1.5, group=Wrap_WomanDay01_Moon, parameterName=WD01_Moon_FRG_CloudsMSK_Radius}
  n78 = ScalarParameter() {defaultValue=2.1, group=Wrap_WomanDay01_Moon, parameterName=WD01_Moon_FRG_CloudsMSK_Density}
  n79 = Multiply(
   A=Multiply(A=Multiply(A=@n53, B=@n76) {constB=15}, B=@n59)
   B=OneMinus(None=MaterialFunctionCall(UVs=@n28, Radius=@n77, Density=@n78) {materialFunction=&RadialGradientExponential}.RadialGradientExponential)
  )
  n80 = LinearInterpolate(
   A=NamedRerouteUsage()
   B=LinearInterpolate(A=LinearInterpolate(A=@n25, B=@n33, Alpha=@n34), B=LinearInterpolate(A=@n71, B=@n72, Alpha=@n75), Alpha=Saturate(None=@n79))
   Alpha=NamedRerouteUsage()
  )
  Add(A=Multiply(A=NamedRerouteUsage(), B=@n24) {constB=-1}, B=Multiply(A=OneMinus(None=@n24), B=@n80))
  n81 = VectorParameter() {group=Wrap_WomanDay01_Stars__FRG, parameterName=WD01_Stars_Speed_Frg, defaultValue=<rgba 0.05 -0.015 0 1>}
  n82 = Multiply(A=Time(), B=@n81.RGB)
  n83 = NamedRerouteUsage()
  n84 = Multiply(
   A=Add(A=@n83, B=ComponentMask(None=@n82) {g=true, r=true}) {constB=0}
   B=ScalarParameter() {defaultValue=20, group=Wrap_WomanDay01_Stars__FRG, parameterName=WD01_Stars_TileRate}
  ) {constB=5}
  n85 = ComponentMask(None=@n84) {r=true}
  n86 = Multiply(A=@n85) {constB=0.5}
  n87 = ComponentMask(None=@n84) {g=true}
  n88 = AppendVector(A=If(A=Frac(None=@n86), B=Constant() {r=0.5}, "A > B"=Add(A=@n87) {constB=0.5}, "A < B"=@n87), B=@n85)
  n89 = TextureSample(UVs=Divide(A=Floor(None=@n88)) {constB=25}, Tex=NamedRerouteUsage()) {samplerSource=SSM_Wrap_WorldGroupSettings}
  Add(A=@n89.A) {constB=0.1}
  Add()
  n90 = TextureSample() {samplerType=SAMPLERTYPE_Masks, texture=/Game/Weapons/FORT_Rifles/Textures/T_Scar_CM}
  n91 = FunctionInput(Preview=@n90.R) {bUsePreviewValueAsDefault=true, inputName=Mask Primary, inputType=FunctionInput_Scalar, sortPriority=24, previewValue=<1 0 0 1>}
  n92 = FunctionInput(Preview=@n90.B) {bUsePreviewValueAsDefault=true, inputName=Mask Secondary, inputType=FunctionInput_Scalar, sortPriority=24, previewValue=<1 0 0 1>}
  n93 = BlendMaterialAttributes(A=NamedRerouteUsage(), B=NamedRerouteUsage(), Alpha=Saturate(None=Add(A=@n91, B=@n92))) {vertexAttributeBlendType=UseA}
  BlendMaterialAttributes(A=@n93, B=NamedRerouteUsage(), Alpha=@n92) {vertexAttributeBlendType=UseA}
  ComponentMask() {g=true, r=true}
  ComponentMask() {g=true, r=true}
  n94 = Constant3Vector() {constant=<rgba 0 1 0 0>}
  n95 = MaterialFunctionCall(Float4=NamedRerouteUsage()) {materialFunction=&BreakOutFloat4Components}
  n96 = LinearInterpolate(B=@n94, Alpha=@n95.G)
  ComponentMask(None=LinearInterpolate(A=LinearInterpolate(A=@n94, B=@n96, Alpha=@n95.R), B=@n94, Alpha=@n95.B)) {g=true, r=true}
  ComponentMask(None=@n63) {r=true}
  Constant3Vector() {constant=<rgba 0 0 1 0>}
  Constant4Vector()
  n97 = TextureSample() {samplerType=SAMPLERTYPE_Normal, texture=/Game/Weapons/FORT_Rifles/T_SCAR_N}
  n98 = Constant3Vector() {constant=<rgba 1 0 0 0>}
  n99 = GetMaterialAttributes(None=FunctionInput(Preview=SetMaterialAttributes!(MaterialAttributes, Normal=@n97.RGB, BaseColor=@n98) {attributeSetTypes [#1, #2]}) {bUsePreviewValueAsDefault=true, inputName=Material, inputType=FunctionInput_MaterialAttributes}) {attributeGetTypes [
    #2
    #3
    #1
    #4
   ]}
  FunctionOutput(None=StaticSwitchParameter(True=@n93, False=NamedRerouteDeclaration(None=@n99.MaterialAttributes) {name=Root_Material, nodeColor=<rgba 0.7812 1 0 1>}) {desc=TopSide Projection Base - Emissive pan Mask, group=Customization_Wrap, parameterName=Wrap_WomanDay01})
  n100 = VectorParameter() {group=Wrap_WomanDay01_Stars_BkG, parameterName=WD01_Stars_Color_Bck_AB, defaultValue=<rgba 0.1196 0.1499 0.7917 1>}
  n101 = VectorParameter() {group=Wrap_WomanDay01_Stars_BkG, parameterName=WD01_Stars_Color_Bck_AA, defaultValue=<rgba 0.9375 0.9357 0.4297 1>}
  n102 = ComponentMask(None=NamedRerouteUsage()) {r=true}
  n103 = ScalarParameter() {defaultValue=4, group=Wrap_WomanDay01_Stars_BkG, parameterName=WD01_Stars_Grad_Falloff_Bck}
  n104 = Power(Base=@n102, Exp=@n103)
  LinearInterpolate(A=@n100.RGB, B=@n101.RGB, Alpha=@n104)
  n105 = ComponentMask(None=NamedRerouteUsage()) {r=true}
  n106 = ScalarParameter() {defaultValue=4, group=Wrap_WomanDay01_Stars__FRG, parameterName=WD01_Stars_Grad_Falloff}
  n107 = VectorParameter() {group=Wrap_WomanDay01_Stars__FRG, parameterName=WD01_Stars_Color_AA, defaultValue=<rgba 0.9375 0.9344 0.0448 1>}
  n108 = OneMinus(None=@n105)
  n109 = VectorParameter() {group=Wrap_WomanDay01_Stars__FRG, parameterName=WD01_Stars_Color_AB, defaultValue=<rgba 0 0.0207 0.4583 1>}
  n110 = Add(A=Multiply(A=Power(Base=@n105, Exp=@n106), B=@n107.RGB), B=Multiply(A=Power(Base=@n108, Exp=@n106), B=@n109.RGB))
  n111 = Multiply(A=@n110) {constB=5}
  n112 = Frac(None=@n88)
  n113 = Time()
  n114 = ScalarParameter() {defaultValue=-0.15, group=Wrap_WomanDay01_Stars__FRG, parameterName=WD01_Stars_Rotation_Speed_Min}
  n115 = Multiply(A=@n113, B=LinearInterpolate(A=@n114, B=Multiply(A=@n114) {constB=0.5}, Alpha=@n89.A))
  n116 = Add(
   A=Multiply(A=@n89.A) {constB=3}
   B=Multiply(A=@n113, B=ScalarParameter() {defaultValue=-0.15, group=Wrap_WomanDay01_Stars__FRG, parameterName=HeadC_Background_A_Fade_Speed})
  )
  n117 = Saturate(None=LinearInterpolate(Alpha=Frac(None=@n116)) {constA=-2, constB=1.1})
  n118 = TextureSample(
   UVs=MaterialFunctionCall(
    UVs=MaterialFunctionCall(UVs=@n112, "Rotation Angle (0-1)"=@n115) {materialFunction=&CustomRotator}.Rotated Values
    Texture Scale=@n117
   ) {materialFunction=&ScaleUVsByCenter}.UVs
   Tex=NamedRerouteUsage()
  ) {samplerSource=SSM_Clamp_WorldGroupSettings}
  LinearInterpolate(A=@n110, B=@n111, Alpha=@n118.R)
  MaterialFunctionCall(UVs=@n9, Rotation Center=Constant2Vector() {g=0.5, r=0.5}, "Rotation Angle (0-1)"=Constant() {r=0.5}) {materialFunction=&CustomRotator}
  MaterialFunctionCall!(UVs=@n83, Texture Size=ScalarParameter() {defaultValue=0.2, group=Wrap_WomanDay01_Stars__FRG, parameterName=WD01_Stars_MipBias}) {materialFunction=/Engine/Functions/Engine_MaterialFunctions01/Texturing/ComputeMipLevel}
  Multiply(A=@n64) {constB=2}
  n119 = Distance(A=@n112, B=Constant() {r=0.5})
  n120 = OneMinus(None=Floor(None=Multiply(A=@n119) {constB=2.1}))
  n121 = Add(A=@n117) {constB=0.25}
  n122 = Multiply(A=@n118.R, B=Multiply(A=@n120, B=Round(None=@n121)))
  Multiply(A=@n118.R, B=@n122)
  n123 = Floor(None=@n22)
  n124 = NamedRerouteUsage()
  n125 = TextureSample(UVs=Divide(A=@n123) {constB=25}, Tex=@n124) {samplerSource=SSM_Wrap_WorldGroupSettings}
  Multiply(A=@n125.A)
  Multiply(A=TextureSample(UVs=@n88) {texture=/Game/Weapons/WeaponSkins/Wraps/Textures/UVCheaker}.R)
  Multiply(A=@n120, B=@n117)
  n126 = ScalarParameter() {defaultValue=-0.15, group=Wrap_WomanDay01_Stars__FRG, parameterName=WD01_Stars_Rotation_Speed_Min}
  n127 = Time()
  n128 = Multiply(A=LinearInterpolate(A=@n126, B=Multiply(A=@n126) {constB=-1}, Alpha=@n125.A), B=@n127)
  n129 = TextureSample!(
   UVs=MaterialFunctionCall(
    UVs=MaterialFunctionCall(UVs=@n23, "Rotation Angle (0-1)"=@n128) {materialFunction=&CustomRotator}.Rotated Values
    Texture Scale=LinearInterpolate(Alpha=@n125.A) {constA=0.1}
   ) {materialFunction=&ScaleUVsByCenter}.UVs
   Tex=NamedRerouteUsage()
   Bias=MaterialFunctionCall(UVs=@n18, Texture Size=ScalarParameter() {defaultValue=0.2, group=Wrap_WomanDay01_Stars__FRG, parameterName=WD01_Stars_MipBias}) {materialFunction=&ComputeMipLevel}.MipLevel
   Apply View MipBias
  ) {mipValueMode=TMVM_MipBias}
  n130 = Multiply(A=@n125.A) {constB=20}
  n131 = Multiply(A=@n127, B=ScalarParameter() {defaultValue=-0.15, group=Wrap_WomanDay01_Stars__FRG, parameterName=HeadC_Background_A_Fade_Speed})
  n132 = OneMinus(None=Frac(None=Add(A=@n130, B=@n131)))
  Multiply(A=@n129.G, B=Saturate(None=LinearInterpolate(Alpha=@n132) {constA=-0.5}))
  Multiply(A=@n89.A, B=@n117)
  Multiply(A=@n129.G)
  Multiply(A=TextureSample!(UVs=@n14.Rotated Values, Tex, Apply View MipBias) {texture=/Game/Weapons/WeaponSkins/Wraps/Textures/UVCheaker_Clamped}.RGB)
  Multiply(A=@n33, B=Multiply(A=@n31, B=@n34))
  n133 = VectorParameter() {group=Wrap_WomanDay01_Clouds_BKG, parameterName=WD01_BackG_Color_Darken, defaultValue=<rgba 0.2911 0.3408 0.6562 1>}
  Multiply(A=@n133.RGB)
  n134 = ComponentMask(None=NamedRerouteUsage()) {r=true}
  n135 = ScalarParameter() {defaultValue=4, group=Wrap_WomanDay01_Stars__FRG, parameterName=WD01_Stars_Grad_Falloff}
  n136 = VectorParameter() {group=Wrap_WomanDay01_Stars__FRG, parameterName=WD01_Stars_Color_AA, defaultValue=<rgba 0.9375 0.9344 0.0448 1>}
  n137 = OneMinus(None=@n134)
  n138 = VectorParameter() {group=Wrap_WomanDay01_Stars__FRG, parameterName=WD01_Stars_Color_AB, defaultValue=<rgba 0 0.0207 0.4583 1>}
  n139 = Add(A=Multiply(A=Power(Base=@n134, Exp=@n135), B=@n136.RGB), B=Multiply(A=Power(Base=@n137, Exp=@n135), B=@n138.RGB))
  Multiply(A=@n129.G, B=LinearInterpolate(A=@n139, B=Multiply(A=@n139) {constB=5}, Alpha=@n129.R))
  n140 = VectorParameter() {group=Wrap_WomanDay01_Stars_BkG, parameterName=WD01_Stars_Speed_BckG_A, defaultValue=<rgba 0.035 -0.015 0 0>}
  n141 = ComponentMask(None=Multiply(A=Time(), B=@n140.RGB)) {g=true, r=true}
  n142 = Multiply(A=NamedRerouteUsage(), B=ScalarParameter() {defaultValue=2, group=Wrap_WomanDay01_Stars_BkG, parameterName=WD01_Stars_TileRate_BckG_A})
  Multiply(A=TextureSample(UVs=Add(A=@n141, B=@n142), Tex=NamedRerouteUsage()) {samplerSource=SSM_Wrap_WorldGroupSettings}.R)
  n143 = Power(Base=OneMinus(None=@n102), Exp=@n103)
  Multiply(A=Add(A=Multiply(A=@n104, B=@n101.RGB), B=Multiply(A=@n143, B=@n100.RGB))) {constB=0.75}
  n144 = TextureObjectParameter() {
   group=Wrap_WomanDay01
   parameterName=WD01_Texture_Clouds
   samplerType=SAMPLERTYPE_Masks
   texture=&T_BlackMask
  }
  NamedRerouteDeclaration(None=@n144) {name=Texture_CloudsMsk, nodeColor=<rgba 0 1 0.7422 1>}
  n145 = Multiply(A=@n14.Rotated Values, B=NamedRerouteUsage())
  NamedRerouteDeclaration(None=@n145) {name=Side_Clouds_Uvs, nodeColor=<rgba 1 0.6328 0 1>}
  NamedRerouteDeclaration() {name=Texture_StarsMsk, nodeColor=<rgba 1 0 0.0415 1>}
  NamedRerouteDeclaration(None=Multiply(A=@n145, B=ScalarParameter() {defaultValue=0.75, group=Wrap_WomanDay01_Clouds_BKG, parameterName=WD01_Side_ScaleClouds_BckG}) {constB=0.75}) {name=Side_CloudsBK_Uvs, nodeColor=<rgba 1 0.6328 0 1>}
  n146 = VectorParameter() {group=Wrap_WomanDay01, parameterName=WD01_Master_Emissive_Brightness, defaultValue=<rgba 0.5 0.5 0.5 1>}
  n147 = Multiply(A=@n80, B=@n146.RGB)
  n148 = Multiply(A=@n147, B=ScalarParameter() {defaultValue=3, group=Wrap_WomanDay01, parameterName=WD01_Edge_Brightness}) {constB=5}
  n149 = NamedRerouteUsage()
  n150 = Saturate(None=Multiply(A=Add(A=@n149) {constB=-0.5}) {constB=2})
  n151 = ScalarParameter() {defaultValue=0.5, group=Wrap_WomanDay01, parameterName=WD01_TOD_Value_Morning}
  n152 = ScalarParameter() {defaultValue=1, group=Wrap_WomanDay01, parameterName=WD01_TOD_Value_Day}
  n153 = ScalarParameter() {defaultValue=0.5, group=Wrap_WomanDay01, parameterName=WD01_TOD_Value_Evening}
  n154 = ScalarParameter() {defaultValue=0.25, group=Wrap_WomanDay01, parameterName=WD01_TOD_Value_Night}
  n155 = SetMaterialAttributes!(
   MaterialAttributes=NamedRerouteUsage()
   BaseColor=VectorParameter() {group=Wrap_WomanDay01, parameterName=WD01_BaseColor, sortPriority=10, defaultValue=<rgba 0 0 0 1>}.RGB
   Roughness=ScalarParameter() {defaultValue=0.35, group=Wrap_WomanDay01, parameterName=WD01_Roughness, sortPriority=11}
   Specular=ScalarParameter() {defaultValue=0.5, group=Wrap_WomanDay01, parameterName=WD01_Specular, sortPriority=12}
   EmissiveColor=Multiply(
    A=LinearInterpolate(A=@n147, B=@n148, Alpha=@n150)
    B=MaterialFunctionCall(Value_Morning=@n151, Value_Day=@n152, Value_Evening=@n153, Value_Night=@n154) {materialFunction=&Scalar_Time_of_Day}.Result
   )
   Metallic=ScalarParameter() {group=Wrap_WomanDay01, parameterName=WD01_Metalic, sortPriority=12}
   Normal
  ) {attributeSetTypes [
    #2
    #3
    #5
    #6
    #4
    #1
   ]}
  NamedRerouteDeclaration(None=@n155) {name=Base_Material, nodeColor=<rgba 1 0.7031 0 1>}
  NamedRerouteDeclaration(None=FunctionInput(Preview=Constant4Vector() {constant=<rgba -0.25 0 0 0.2>}.RGBA) {bUsePreviewValueAsDefault=true, inputName=Object-SafeZone-Front, inputType=FunctionInput_Vector4, sortPriority=102}) {name=SafeZone_Frnt, nodeColor=<rgba 0.9922 1 0 1>}
  n156 = ScalarParameter() {group=Wrap_WomanDay01_UVs, parameterName=WD01_Side_Rotation_Weapon}
  n157 = MaterialFunctionCall(Float4=NamedRerouteUsage()) {materialFunction=&BreakOutFloat4Components}
  n158 = LinearInterpolate(
   A=ScalarParameter() {group=Wrap_WomanDay01_UVs, parameterName=WD01_Side_Rotation_Vehicle}
   B=ScalarParameter() {group=Wrap_WomanDay01_UVs, parameterName=WD01_Side_Rotation_VehicleSide}
   Alpha=@n157.G
  )
  n159 = ScalarParameter() {defaultValue=0.25, group=Wrap_WomanDay01_UVs, parameterName=WD01_Side_Rotation_Mech}
  n160 = ScalarParameter() {group=Wrap_WomanDay01_UVs, parameterName=WD01_Side_Rotation_Character}
  NamedRerouteDeclaration(None=LinearInterpolate(A=LinearInterpolate(A=@n156, B=@n158, Alpha=@n157.R), B=LinearInterpolate(A=@n159, B=@n160, Alpha=@n157.A), Alpha=@n157.B)) {name=Primary_Rotation, nodeColor=<rgba 1 0.7266 0 1>}
  n161 = StaticBool() {value=true}
  n162 = MaterialFunctionCall(Float4=NamedRerouteUsage()) {materialFunction=&BreakOutFloat4Components}
  n163 = NamedRerouteUsage()
  n164 = NamedRerouteUsage()
  n165 = MaterialFunctionCall(
   For_Safe_Zones=@n161
   Caculated_Bounds=NamedRerouteUsage()
   Pre_Skinned_Normals=NamedRerouteUsage()
   Is Vehicle=@n162.R
   Is VehicleSide=@n162.R
   Is Mech=@n162.R
   UV_Rotation Bypass=Add(A=@n163, B=ScalarParameter() {defaultValue=-0.05, group=Wrap_WomanDay01, parameterName=WD01_Moon_Rotation, sortPriority=9}) {constB=0}
   Uvs_Offset=ComponentMask(None=@n164) {b=true, g=true, r=true}
   Uvs_TileRate=ComponentMask(None=@n164) {a=true}
  ) {materialFunction=&MF_Projection_Side_Normal}
  NamedRerouteDeclaration(None=@n165.UVs) {name=UVs_SafeZone, nodeColor=<rgba 1 0.1057 0 1>}
  NamedRerouteDeclaration(None=@n165.Safe_Zone_Mask) {name=UVs_SafeZone_Msk, nodeColor=<rgba 0.6172 1 0 1>}
  n166 = TextureObjectParameter() {
   group=Wrap_WomanDay01
   parameterName=WD01_Texture_Moon
   samplerType=SAMPLERTYPE_Masks
   texture=&T_BlackMask
  }
  NamedRerouteDeclaration(None=@n166) {name=Texture_MoonMsk, nodeColor=<rgba 0 1 0.7422 1>}
  NamedRerouteDeclaration(None=MaterialFunctionCall(
    UVs=@n28
    Radius=ScalarParameter() {defaultValue=1.5, group=Wrap_WomanDay01_Moon, parameterName=WD01_Moon_GlowMsk_Radius}
    Density=ScalarParameter() {defaultValue=1.25, group=Wrap_WomanDay01_Moon, parameterName=WD01_Moon_GlowMsk_Density}
   ) {materialFunction=&RadialGradientExponential}.RadialGradientExponential) {name=Moon_Glow_Msk, nodeColor=<rgba 0.5234 0 1 1>}
  n167 = ScalarParameter() {defaultValue=1, group=Wrap_WomanDay01_Clouds__FRG, parameterName=WD01_Side_ScaleClouds_Vehicle, sortPriority=150}
  n168 = ScalarParameter() {defaultValue=1, group=Wrap_WomanDay01_Clouds__FRG, parameterName=WD01_Side_ScaleClouds_Vehicle_Side, sortPriority=150}
  n169 = MaterialFunctionCall(Float4=NamedRerouteUsage()) {materialFunction=&BreakOutFloat4Components}
  n170 = LinearInterpolate(
   A=LinearInterpolate(
    A=ScalarParameter() {defaultValue=1, group=Wrap_WomanDay01_Clouds__FRG, parameterName=WD01_Side_ScaleClouds_Weapon, sortPriority=150}
    B=LinearInterpolate(A=@n167, B=@n168, Alpha=@n169.G)
    Alpha=@n169.R
   )
   B=LinearInterpolate(
    A=ScalarParameter() {defaultValue=1, group=Wrap_WomanDay01_Clouds__FRG, parameterName=WD01_Side_ScaleClouds_Mech, sortPriority=150}
    B=ScalarParameter() {defaultValue=1, group=Wrap_WomanDay01_Clouds__FRG, parameterName=WD01_Side_ScaleClouds_Character, sortPriority=150}
    Alpha=@n169.A
   )
   Alpha=@n169.B
  )
  n171 = ScalarParameter() {defaultValue=1, group=Wrap_WomanDay01_Clouds__FRG, parameterName=WD01_Side_ScaleClouds_Backpack, sortPriority=150}
  n172 = MaterialFunctionCall(Float2=NamedRerouteUsage()) {materialFunction=&BreakOutFloat2Components}
  NamedRerouteDeclaration(None=LinearInterpolate(
    A=LinearInterpolate(A=@n170, B=@n171, Alpha=@n172.R)
    B=ScalarParameter() {defaultValue=1, group=Wrap_WomanDay01_Clouds__FRG, parameterName=WD01_Side_ScaleClouds_Pickaxe, sortPriority=150}
    Alpha=@n172.G
   )) {name=Side_Clouds_Scale, nodeColor=<rgba 0 0.7187 1 1>}
  n173 = VectorParameter() {group=Wrap_WomanDay01_Moon, parameterName=WD01_MoonGlow_ColorBKG, defaultValue=<rgba 5.7574 4.1522 8 5>}
  n174 = Time()
  n175 = VectorParameter() {group=Wrap_WomanDay01_Stars_BkG, parameterName=WD01_Stars_twinkle_Speed, defaultValue=<rgba 0.09 0.033 0 0>}
  n176 = Add(A=Multiply(A=NamedRerouteUsage()) {constB=0.1}, B=ComponentMask(None=Multiply(A=@n174, B=@n175.RGB)) {g=true, r=true})
  n177 = NamedRerouteUsage()
  n178 = TextureSample(
   UVs=Multiply(A=NamedRerouteUsage(), B=ScalarParameter() {defaultValue=2, group=Wrap_WomanDay01_Stars_BkG, parameterName=WD01_Stars_TileRate_BckG_B})
   Tex=@n177
  ) {samplerSource=SSM_Wrap_WorldGroupSettings}
  n179 = Multiply(A=ScalarParameter() {defaultValue=2, group=Wrap_WomanDay01_Stars_BkG, parameterName=WD01_Stars_twinkle_Warp}, B=@n178.R)
  n180 = Multiply(A=TextureSample(UVs=Add(A=@n176, B=@n179), Tex=@n177) {samplerSource=SSM_Wrap_WorldGroupSettings}.B, B=@n178.A)
  n181 = ScalarParameter() {defaultValue=4, group=Wrap_WomanDay01_Stars_BkG, parameterName=WD01_Stars_twinkle_Brightness}
  n182 = Multiply(A=@n178.R, B=@n178.R) {constB=0.05}
  n183 = ScalarParameter() {defaultValue=0.5, group=Wrap_WomanDay01_Stars_BkG, parameterName=WD01_Stars_Brightness}
  n184 = VectorParameter() {group=Wrap_WomanDay01_Stars_BkG, parameterName=WD01_Stars_Color_Bck_01, defaultValue=<rgba 0.125 0.125 0.125 1>}
  n185 = Multiply(A=Add(A=Multiply(A=@n180, B=@n181) {constB=5}, B=Multiply(A=@n182, B=@n183) {constB=0.25}), B=@n184.RGB)
  n186 = NamedRerouteUsage()
  NamedRerouteDeclaration(None=LinearInterpolate(A=@n185, B=Multiply(A=@n173.RGB, B=@n185), Alpha=@n186)) {name=Tri_Msk_Color, nodeColor=<rgba 0.2891 1 0 1>}
  n187 = MaterialFunctionCall(Color=Abs(None=NamedRerouteUsage())) {materialFunction=&SplitComponents}
  n188 = MaterialFunctionCall(Float4=NamedRerouteUsage()) {materialFunction=&BreakOutFloat4Components}
  n189 = LinearInterpolate(A=@n187.R, B=LinearInterpolate(A=@n187.B, B=@n187.B, Alpha=@n188.G), Alpha=@n188.R)
  NamedRerouteDeclaration(None=MaterialFunctionCall(
    In=LinearInterpolate(A=@n189, B=@n187.G, Alpha=@n188.B)
    Contrast=ScalarParameter() {defaultValue=1, group=Wrap_WomanDay01, parameterName=WD01_TriPlanerMsk_Contrast, sortPriority=9}
   ) {materialFunction=&CheapContrast}.Result) {name=Tri_Mask, nodeColor=<rgba 1 0.4688 0 1>}
  n190 = VectorParameter() {group=Wrap_WomanDay01_UVs, parameterName=WD01_Side_Offset_Vehicle_01, defaultValue=<rgba 0 0 0 1>}
  n191 = VectorParameter() {group=Wrap_WomanDay01_UVs, parameterName=WD01_Side_Offset_VehicleSide_01, defaultValue=<rgba 0 1 0 1>}
  n192 = MaterialFunctionCall(Float4=NamedRerouteUsage()) {materialFunction=&BreakOutFloat4Components}
  n193 = LinearInterpolate(
   A=LinearInterpolate(
    A=VectorParameter() {group=Wrap_WomanDay01_UVs, parameterName=WD01_Side_Offset_Weapon_01, defaultValue=<rgba 0 1 0 1>}.RGB
    B=LinearInterpolate(A=@n190.RGB, B=@n191.RGB, Alpha=@n192.G)
    Alpha=@n192.R
   )
   B=LinearInterpolate(
    A=VectorParameter() {group=Wrap_WomanDay01_UVs, parameterName=WD01_Side_Offset_Mech_01, defaultValue=<rgba 0 1 0 1>}.RGB
    B=VectorParameter() {group=Wrap_WomanDay01_UVs, parameterName=WD01_Side_Offset_Character_01, defaultValue=<rgba 0 1 0 1>}.RGB
    Alpha=@n192.A
   )
   Alpha=@n192.B
  )
  n194 = VectorParameter() {group=Wrap_WomanDay01_UVs, parameterName=WD01_Side_Offset_Pickaxe_01, defaultValue=<rgba 0 1 0 1>}
  n195 = MaterialFunctionCall(Float2=NamedRerouteUsage()) {materialFunction=&BreakOutFloat2Components}
  NamedRerouteDeclaration(None=ComponentMask(None=LinearInterpolate(A=@n193, B=@n194.RGB, Alpha=@n195.G)) {g=true, r=true}) {name=SidePJ_offset, nodeColor=<rgba 0 0.625 1 1>}
  n196 = MaterialFunctionCall(Float3=Abs(None=NamedRerouteUsage())) {materialFunction=&BreakOutFloat3Components}
  n197 = Constant() {r=0.5}
  n198 = NamedRerouteUsage()
  n199 = ComponentMask(None=@n198) {b=true, desc=YZ, g=true}
  n200 = ComponentMask(None=@n198) {b=true, desc=XZ, r=true}
  NamedRerouteDeclaration(None=If(
    A=@n196.B
    B=@n197
    "A > B"=ComponentMask(None=@n198) {desc=YX, g=true, r=true}
    "A < B"=If(A=@n196.R, B=@n197, "A > B"=@n199, "A < B"=@n200)
   )) {name=UVS_Cheap_Tri, nodeColor=<rgba 1 0 0.8437 1>}
  NamedRerouteDeclaration(None=@n111) {name=Stars_FRG, nodeColor=<rgba 1 0 0.8906 1>}
  NamedRerouteDeclaration(None=Multiply(A=@n122, B=NamedRerouteUsage())) {name=Stars_FRG_Alpha, nodeColor=<rgba 1 0 0.8906 1>}
  n201 = VectorParameter() {group=Wrap_WomanDay01_Clouds_BKG, parameterName=WD01_BackG_Color_BA, defaultValue=<rgba 0.1771 0.1241 0.1107 1>}
  n202 = VectorParameter() {group=Wrap_WomanDay01_Clouds_BKG, parameterName=WD01_BackG_Color_BC, defaultValue=<rgba 0.0865 0.0254 0.3125 1>}
  n203 = VectorParameter() {group=Wrap_WomanDay01_Clouds_BKG, parameterName=WD01_BackG_Color_BB, defaultValue=<rgba 0.2581 0.5881 0.6979 1>}
  n204 = NamedRerouteUsage()
  n205 = NamedRerouteUsage()
  n206 = LinearInterpolate(
   A=@n202.RGB
   B=@n203.RGB
   Alpha=TextureSample(UVs=Panner(Coordinate=@n204, Time=@n2) {speedX=-0.025, speedY=-0.002}, Tex=@n205) {samplerSource=SSM_Wrap_WorldGroupSettings, samplerType=SAMPLERTYPE_Masks}.B
  )
  n207 = TextureSample(UVs=Panner(Coordinate=Multiply(A=@n204) {constB=1.25}, Time=@n2) {speedX=0.033, speedY=-0.0033}, Tex=@n205) {samplerSource=SSM_Wrap_WorldGroupSettings, samplerType=SAMPLERTYPE_Masks}
  n208 = Add(A=NamedRerouteUsage(), B=NamedRerouteUsage())
  n209 = Multiply(A=Power(Base=ComponentMask(None=@n208) {g=true}) {constExponent=1}) {constB=0.75}
  n210 = NamedRerouteUsage()
  n211 = TextureSample(UVs=@n11, Tex=@n210) {samplerSource=SSM_Wrap_WorldGroupSettings, samplerType=SAMPLERTYPE_Masks}
  n212 = TextureSample(UVs=@n10, Tex=@n210) {samplerSource=SSM_Wrap_WorldGroupSettings, samplerType=SAMPLERTYPE_Masks}
  n213 = LinearInterpolate(Alpha=Multiply(A=Add(A=@n211.R, B=@n212.G), B=@n212.R)) {constB=0.65}
  n214 = Multiply(A=LinearInterpolate(B=LinearInterpolate(A=@n201.RGB, B=@n206, Alpha=@n207.B), Alpha=Multiply(A=@n209, B=@n213)), B=@n133.RGB) {constB=0.5}
  NamedRerouteDeclaration(None=LinearInterpolate(A=@n214, B=Multiply(A=@n214, B=@n173.RGB), Alpha=@n186)) {name=Clouds_Bck, nodeColor=<rgba 0.9688 1 0 1>}
  NamedRerouteDeclaration(None=Add(A=NamedRerouteUsage(), B=@n185)) {name="Clouds\\Stars_Bck", nodeColor=<rgba 0 1 0.7656 1>}
  NamedRerouteDeclaration(None=@n14.Rotated Values) {name=Side_Uvs, nodeColor=<rgba 1 0.6328 0 1>}
  n215 = ScalarParameter() {group=Wrap_WomanDay01_Clouds__FRG, parameterName=WD01_Clouds_FRG_Offset_Weapon, sortPriority=100}
  n216 = MaterialFunctionCall(Float4=NamedRerouteUsage()) {materialFunction=&BreakOutFloat4Components}
  n217 = LinearInterpolate(
   A=ScalarParameter() {group=Wrap_WomanDay01_Clouds__FRG, parameterName=WD01_Clouds_FRG_Offset_Vehicle, sortPriority=100}
   B=ScalarParameter() {group=Wrap_WomanDay01_Clouds__FRG, parameterName=WD01_Clouds_FRG_Offset_VehicleSide, sortPriority=100}
   Alpha=@n216.G
  )
  n218 = ScalarParameter() {defaultValue=0.25, group=Wrap_WomanDay01_Clouds__FRG, parameterName=WD01_Clouds_FRG_Offset_Mech, sortPriority=100}
  n219 = ScalarParameter() {group=Wrap_WomanDay01_Clouds__FRG, parameterName=WD01_Clouds_FRG_Offset_Character, sortPriority=100}
  NamedRerouteDeclaration(None=LinearInterpolate(A=LinearInterpolate(A=@n215, B=@n217, Alpha=@n216.R), B=LinearInterpolate(A=@n218, B=@n219, Alpha=@n216.A), Alpha=@n216.B)) {name=FRG_Offset, nodeColor=<rgba 1 0.7266 0 1>}
  n220 = VectorParameter() {group=Wrap_WomanDay01, parameterName=WD01_Accent_BaseColor, sortPriority=10, defaultValue=<rgba 0 0 0 1>}
  n221 = SetMaterialAttributes(
   MaterialAttributes=NamedRerouteUsage()
   BaseColor=@n220.RGB
   Roughness=ScalarParameter() {defaultValue=0.35, group=Wrap_WomanDay01, parameterName=WD01_Accent_Roughness, sortPriority=10}
   Specular=ScalarParameter() {defaultValue=0.5, group=Wrap_WomanDay01, parameterName=WD01_Accent_Specular, sortPriority=10}
   Metallic=ScalarParameter() {defaultValue=1, group=Wrap_WomanDay01, parameterName=WD01_Accent_Metalic, sortPriority=10}
  ) {attributeSetTypes [
    #2
    #3
    #5
    #6
    #4
    #1
   ]}
  NamedRerouteDeclaration(None=@n221) {name=Accent_Material, nodeColor=<rgba 1 0 0.6328 1>}
  n222 = FunctionInput(Preview=@n90.G) {bUsePreviewValueAsDefault=true, inputName=Mask Crevise, inputType=FunctionInput_Scalar, sortPriority=24, previewValue=<1 0 0 1>}
  NamedRerouteDeclaration(None=@n222) {name=Crevise_map, nodeColor=<rgba 1 0.6797 0 1>}
  n223 = MaterialFunctionCall(Float4=NamedRerouteUsage()) {materialFunction=&BreakOutFloat4Components}
  n224 = NamedRerouteUsage()
  NamedRerouteDeclaration(None=MaterialFunctionCall(
    For_Safe_Zones=@n161
    Caculated_Bounds=NamedRerouteUsage()
    Pre_Skinned_Normals=NamedRerouteUsage()
    Is Vehicle=@n223.R
    Is VehicleSide=@n223.R
    Is Mech=@n223.R
    UV_Rotation Bypass=@n163
    Uvs_Offset=ComponentMask(None=@n224) {b=true, g=true, r=true}
    Uvs_TileRate=ComponentMask(None=@n224) {a=true}
   ) {materialFunction=&MF_Projection_Side_Normal}.UVs) {name=UVs_SafeZone_Clouds, nodeColor=<rgba 0.6172 1 0 1>}
  n225 = ScalarParameter() {group=Wrap_WomanDay01_Clouds__FRG, parameterName=WD01_Grad_offset_Vehicle, sortPriority=150}
  n226 = ScalarParameter() {group=Wrap_WomanDay01_Clouds__FRG, parameterName=WD01_Grad_offset_Vehicle_Side, sortPriority=150}
  n227 = MaterialFunctionCall(Float4=NamedRerouteUsage()) {materialFunction=&BreakOutFloat4Components}
  n228 = LinearInterpolate(
   A=LinearInterpolate(
    A=ScalarParameter() {group=Wrap_WomanDay01_Clouds__FRG, parameterName=WD01_Grad_offset_Weapon, sortPriority=150}
    B=LinearInterpolate(A=@n225, B=@n226, Alpha=@n227.G)
    Alpha=@n227.R
   )
   B=LinearInterpolate(
    A=ScalarParameter() {group=Wrap_WomanDay01_Clouds__FRG, parameterName=WD01_Grad_offset_Mech, sortPriority=150}
    B=ScalarParameter() {group=Wrap_WomanDay01_Clouds__FRG, parameterName=WD01_Grad_offset_Character, sortPriority=150}
    Alpha=@n227.A
   )
   Alpha=@n227.B
  )
  n229 = ScalarParameter() {group=Wrap_WomanDay01_Clouds__FRG, parameterName=WD01_Grad_offset_Backpack, sortPriority=150}
  n230 = MaterialFunctionCall(Float2=NamedRerouteUsage()) {materialFunction=&BreakOutFloat2Components}
  NamedRerouteDeclaration(None=LinearInterpolate(
    A=LinearInterpolate(A=@n228, B=@n229, Alpha=@n230.R)
    B=ScalarParameter() {group=Wrap_WomanDay01_Clouds__FRG, parameterName=WD01_Grad_offset_Pickaxe, sortPriority=150}
    Alpha=@n230.G
   )) {name=Side_Clouds_Grad_offset, nodeColor=<rgba 0 0.7188 1 1>}
  NamedRerouteDeclaration(None=FunctionInput(Preview=Constant4Vector().RGBA) {bUsePreviewValueAsDefault=true, inputName=IsState, inputType=FunctionInput_Vector4, sortPriority=4}) {name=IsState, nodeColor=<rgba 0.1497 1 0.1633 1>}
  NamedRerouteDeclaration(None=FunctionInput(Preview=Constant2Vector()) {bUsePreviewValueAsDefault=true, inputName=IsState_B, inputType=FunctionInput_Vector2, sortPriority=4}) {name=IsState_B, nodeColor=<rgba 0.1497 1 0.1633 1>}
  n231 = PreSkinnedLocalBounds()
  n232 = Subtract(A=VertexInterpolator(VS=PreSkinnedPosition()).PS, B=@n231.Min)
  n233 = MaterialFunctionCall(Color=Subtract(A=@n231.Max, B=@n231.Max)) {materialFunction=&SplitComponents}
  n234 = Max(A=Max(A=@n233.R, B=@n233.R), B=@n233.B)
  NamedRerouteDeclaration(None=FunctionInput(Preview=Divide(A=@n232, B=@n234)) {bUsePreviewValueAsDefault=true, inputName=Calcated_Bounds, sortPriority=11}) {name=Caclcated_Bounds, nodeColor=<rgba 0 0.9297 1 1>}
  n235 = TextureObjectParameter() {
   group=Wrap_WomanDay01
   parameterName=WD01_Texture_star
   samplerType=SAMPLERTYPE_Masks
   texture=&T_BlackMask
  }
  NamedRerouteDeclaration(None=@n235) {name=Texture_SStarMsk, nodeColor=<rgba 0.009 0.0728 1 1>}
  n236 = PreSkinnedNormal()
  NamedRerouteDeclaration(None=FunctionInput(Preview=VertexInterpolator(VS=@n236).PS) {bUsePreviewValueAsDefault=true, inputName=Pre-Skinned Local Normal, sortPriority=10}) {name=PreSkin_Normal, nodeColor=<rgba 0.8281 1 0 1>}
  NamedRerouteDeclaration(None=@n99.Normal) {name=Root_Normal, nodeColor=<rgba 0.4062 0 1 1>}
  NamedRerouteUsage()
  NamedRerouteUsage()
  NamedRerouteUsage()
  NamedRerouteUsage()
  NamedRerouteUsage()
  n237 = Distance(A=@n23, B=Constant() {r=0.5})
  OneMinus(None=Ceil(None=Add(A=@n237) {constB=-0.48}))
  OneMinus(None=Saturate(None=Multiply(A=@n178.A) {constB=0.1}))
  Saturate(None=@n134)
  Saturate(None=@n110)
  Saturate(None=@n102)
  n238 = NamedRerouteUsage()
  Saturate(None=Divide(
    A=ComponentMask(None=@n238) {r=true}
    B=ScalarParameter() {defaultValue=0.5, group=Wrap_WomanDay01_Clouds__FRG, parameterName=WD01_Cloud_Grad_Falloff}
   ) {constB=0.5})
  TextureCoordinate()
  TextureObjectParameter() {
   group=Wrap_WomanDay01
   parameterName=WD01_Texture_Stars
   samplerType=SAMPLERTYPE_Masks
   texture=&T_BlackMask
  }
  TextureSample(UVs=Divide(A=Add(A=@n123) {constB=2}) {constB=18}, Tex=@n124) {samplerSource=SSM_Wrap_WorldGroupSettings}
  VectorParameter() {group=Wrap_WomanDay01_Stars__FRG, parameterName=WD01_Stars_Speed_Frg, defaultValue=<rgba 0.05 -0.015 0 1>}
  VertexNormalWS()
 ]}
