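# Winter Lights wrap material function that layers animated emissive lights over weapons, vehicles, mechs, characters, pickaxes and backpacks.
# Top/side projection supplies base color, roughness, normal and a packed mask, with offset, scale and rotation lerped by item-state inputs.
# Mask channels warp and gate three weighted animation sets (panning or stepped), crossfaded from static glow by a game-time parameter.
# A static switch enables the result, blended through the primary mask.
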MaterialFunction /Game/Weapons/WeaponSkins/Wraps/MaterialFunctions/MF_Wrap_WLights
graph {nodes [
  n1 = NamedRerouteUsage()
  n2 = NamedRerouteUsage()
  n3 = MaterialFunctionCall(Float4=NamedRerouteUsage()) {materialFunction=&BreakOutFloat4Components}
  n4 = NamedRerouteUsage()
  n5 = TextureObjectParameter() {group=Wrap_WinterLights, parameterName=WNL_Main_Mask, sortPriority=10, texture=&T_BlacksRGB}
  n6 = NamedRerouteUsage()
  n7 = ScalarParameter() {defaultValue=1, group=Wrap_WinterLights, parameterName=WNL_Main_Blend_Contrast, sortPriority=9}
  n8 = ScalarParameter() {group=Wrap_WinterLights, parameterName=WNL_Main_Fill_alternate_Axis_Opacity 01, sortPriority=9}
  n9 = VectorParameter() {group=Wrap_WinterLights, parameterName=WNL_Main_Mask_Fill_alternate_Axis, sortPriority=10, defaultValue=<rgba 0 0 0 1>}
  n10 = AppendVector(A=@n9.RGB, B=@n9.RGB)
  n11 = VectorParameter() {group=Wrap_WinterLights_UV, parameterName=WNL_Offset_Vehicle_01, defaultValue=<rgba 0 0 0 1>}
  n12 = VectorParameter() {group=Wrap_WinterLights_UV, parameterName=WNL_Offset_VehicleSide_01, defaultValue=<rgba 0 0 0 1>}
  n13 = MaterialFunctionCall(Float4=NamedRerouteUsage()) {materialFunction=&BreakOutFloat4Components}
  n14 = LinearInterpolate(
   A=LinearInterpolate(
    A=VectorParameter() {group=Wrap_WinterLights_UV, parameterName=WNL_Offset_Weapon_01, defaultValue=<rgba 0 0 0 1>}.RGB
    B=LinearInterpolate(A=@n11.RGB, B=@n12.RGB, Alpha=@n13.G)
    Alpha=@n13.R
   )
   B=LinearInterpolate(
    A=VectorParameter() {group=Wrap_WinterLights_UV, parameterName=WNL_Offset_Mech_01, defaultValue=<rgba 0 0 0 1>}.RGB
    B=VectorParameter() {group=Wrap_WinterLights_UV, parameterName=WNL_Offset_Character_01, defaultValue=<rgba 0 0 0 1>}.RGB
    Alpha=@n13.A
   )
   Alpha=@n13.B
  )
  n15 = VectorParameter() {group=Wrap_WinterLights_UV, parameterName=WNL_Offset_Pickaxe_01, defaultValue=<rgba 0 0 0 1>}
  n16 = MaterialFunctionCall(Float2=NamedRerouteUsage()) {materialFunction=&BreakOutFloat2Components}
  n17 = VectorParameter() {group=Wrap_WinterLights_UV, parameterName=WNL_Offset_Backpack_01-----------, defaultValue=<rgba 0 0 0 1>}
  n18 = VectorParameter() {group=Wrap_WinterLights_UV, parameterName=WNL_Scale_Weapon_01, defaultValue=<rgba 1 1 1 1>}
  n19 = VectorParameter() {group=Wrap_WinterLights_UV, parameterName=WNL_Scale_Vehicle_01, defaultValue=<rgba 1 1 1 1>}
  n20 = VectorParameter() {group=Wrap_WinterLights_UV, parameterName=WNL_Scale_VehicleSide_01, defaultValue=<rgba 1 1 1 1>}
  n21 = MaterialFunctionCall(Float4=NamedRerouteUsage()) {materialFunction=&BreakOutFloat4Components}
  n22 = LinearInterpolate(
   A=Multiply(A=@n18.RGB, B=@n18.RGB)
   B=LinearInterpolate(A=Multiply(A=@n19.RGB, B=@n19.RGB), B=Multiply(A=@n20.RGB, B=@n20.RGB), Alpha=@n21.G)
   Alpha=@n21.R
  )
  n23 = VectorParameter() {group=Wrap_WinterLights_UV, parameterName=WNL_Scale_Mech_01, defaultValue=<rgba 1 1 1 1>}
  n24 = VectorParameter() {group=Wrap_WinterLights_UV, parameterName=WNL_Scale_Character_01, defaultValue=<rgba 1 1 1 1>}
  n25 = LinearInterpolate(A=Multiply(A=@n23.RGB, B=@n23.RGB), B=Multiply(A=@n24.RGB, B=@n24.RGB), Alpha=@n21.A)
  n26 = VectorParameter() {group=Wrap_WinterLights_UV, parameterName=WNL_Scale_Pickaxe_01, defaultValue=<rgba 1 1 1 1>}
  n27 = MaterialFunctionCall(Float2=NamedRerouteUsage()) {materialFunction=&BreakOutFloat2Components}
  n28 = VectorParameter() {group=Wrap_WinterLights_UV, parameterName=WNL_Scale_Backpack_01-----------, defaultValue=<rgba 1 1 1 1>}
  n29 = LinearInterpolate(
   A=LinearInterpolate(A=LinearInterpolate(A=@n22, B=@n25, Alpha=@n21.B), B=Multiply(A=@n26.RGB, B=@n26.RGB), Alpha=@n27.R)
   B=Multiply(A=@n28.RGB, B=@n28.RGB)
   Alpha=@n27.G
  )
  n30 = Multiply(A=NamedRerouteUsage(), B=@n29)
  n31 = Add(A=LinearInterpolate(A=LinearInterpolate(A=@n14, B=@n15.RGB, Alpha=@n16.R), B=@n17.RGB, Alpha=@n16.G), B=@n30)
  n32 = VectorParameter() {group=Wrap_WinterLights_UV, parameterName=WNL_Rotation_Vehicle_01, defaultValue=<rgba 0 0 0 1>}
  n33 = VectorParameter() {
   desc=must activate TS_Vehicle_Side_01 UVs
   group=Wrap_WinterLights_UV
   parameterName=WNL_Rotation_Vehicle_Side_01
   defaultValue=<rgba 0 0 0 1>
  }
  n34 = MaterialFunctionCall(Float4=NamedRerouteUsage()) {materialFunction=&BreakOutFloat4Components}
  n35 = LinearInterpolate(
   A=LinearInterpolate(
    A=VectorParameter() {group=Wrap_WinterLights_UV, parameterName=WNL_Rotation_Weapon_01, defaultValue=<rgba 0 0 -0.25 1>}.RGB
    B=LinearInterpolate(A=@n32.RGB, B=@n33.RGB, Alpha=@n34.G)
    Alpha=@n34.R
   )
   B=LinearInterpolate(
    A=VectorParameter() {group=Wrap_WinterLights_UV, parameterName=WNL_Rotation_Mech_01, defaultValue=<rgba 0.25 0.25 0 1>}.RGB
    B=VectorParameter() {group=Wrap_WinterLights_UV, parameterName=WNL_Rotation_Character_01, defaultValue=<rgba 0.25 0.25 0 1>}.RGB
    Alpha=@n34.A
   )
   Alpha=@n34.B
  )
  n36 = MaterialFunctionCall(Float4=MaterialFunctionCall!(
    IsCharacter=@n3.A
    IsStatsB=@n4
    Texture=@n5
    Rotation_Axis=@n35
    Caculated_Bounds=@n31
    Pre_Skinned_Normals=@n6
    Blend_Contrast=@n7
    Fill_alternate_Axis_Opacity=@n8
    "Fill_alternate_Axis_Color_A=Roughness"=@n10
    Is_Vehicle=@n3.A
    Is_Mech=@n3.A
    Enable_Fill_Z_Weapon
    Fill_Z_Weapon_opacity
    Fill_Z_Weapon_Color
   ) {materialFunction=&/MF_TopsSideProjection_Base}.Combined) {materialFunction=&BreakOutFloat4Components}
  n37 = Multiply(
   A=OneMinus(None=@n36.G)
   B=ScalarParameter() {defaultValue=0.15, group=Wrap_WinterLights_Emissive, parameterName=WNL_Animation_Light_Warp_Amount}
  )
  Add(
   A=Multiply(A=Multiply(A=@n1, B=@n2), B=ScalarParameter() {defaultValue=3, group=Wrap_WinterLights_Emissive, parameterName=WNL_Animation_Light_tileRate})
   B=@n37
  ) {constB=0}
  n38 = NamedRerouteUsage()
  ComponentMask(None=MaterialFunctionCall(
    IsCharacter=@n3.A
    IsStatsB=@n4
    Texture=@n5
    Rotation_Axis=@n35
    Caculated_Bounds=Add(A=@n30, B=@n38)
    Pre_Skinned_Normals=NamedRerouteUsage()
    Blend_Contrast=@n7
    Fill_alternate_Axis_Opacity=@n8
    "Fill_alternate_Axis_Color_A=Roughness"=@n10
    Is_Vehicle=@n3.A
    Is_Mech=@n3.A
   ) {materialFunction=&MF_TopsSideProjection_Base}.RGB) {r=true}
  Constant4Vector()
  n39 = TextureSample() {samplerType=SAMPLERTYPE_Masks, texture=/Game/Weapons/FORT_Rifles/Textures/T_Scar_CM}
  FunctionInput(Preview=@n39.B) {inputName=Mask Secondary, inputType=FunctionInput_Scalar, sortPriority=24, previewValue=<1 0 0 1>}
  n40 = FunctionInput(Preview=SetMaterialAttributes!(MaterialAttributes, Normal=TextureSample() {samplerType=SAMPLERTYPE_Normal, texture=/Game/Weapons/FORT_Rifles/T_SCAR_N}.RGB) {attributeSetTypes [#1]}) {inputName=Material, inputType=FunctionInput_MaterialAttributes}
  n41 = GetMaterialAttributes(None=@n40) {attributeGetTypes [#2, #3, #1]}
  n42 = TextureObjectParameter() {group=Wrap_WinterLights, parameterName=WNL_Main_BaseColor, sortPriority=10, texture=&T_BlacksRGB}
  n43 = VectorParameter() {group=Wrap_WinterLights, parameterName="WNL_Main_BaseColor_Fill_alternate_Axis=Roughness 01", sortPriority=10}
  n44 = MaterialFunctionCall(
   IsCharacter=@n3.A
   IsStatsB=@n4
   Texture=@n42
   Rotation_Axis=@n35
   Caculated_Bounds=@n31
   Pre_Skinned_Normals=@n6
   Blend_Contrast=@n7
   Fill_alternate_Axis_Opacity=@n8
   "Fill_alternate_Axis_Color_A=Roughness"=AppendVector(A=@n43.RGB, B=@n43.RGB)
   Is_Vehicle=@n3.A
   Is_Mech=@n3.A
  ) {materialFunction=&MF_TopsSideProjection_Base}
  n45 = TextureSample(UVs=@n36.R, Tex=NamedRerouteUsage())
  n46 = Multiply(A=@n45.RGB, B=ScalarParameter() {defaultValue=10, group=Wrap_WinterLights_Emissive, parameterName=WNL_Emissive Brightness})
  n47 = ScalarParameter() {group=Wrap_WinterLights_Emissive, parameterName=Activate SetA}
  n48 = Multiply(A=@n36.B, B=ScalarParameter() {defaultValue=0.5, group=Wrap_WinterLights_Emissive, parameterName=WNL_Animation_SetA_WarpAmount})
  n49 = NamedRerouteUsage()
  n50 = NamedRerouteUsage()
  n51 = VectorParameter() {group=Wrap_WinterLights_UV_animation, parameterName=WNL_Animation_Panning_SetA_Weapon, defaultValue=<rgba 0.2 0 0 1>}
  n52 = VectorParameter() {group=Wrap_WinterLights_UV_animation, parameterName=WNL_Animation_Panning_SetA_Vehicle, defaultValue=<rgba 0 0 0 1>}
  n53 = VectorParameter() {group=Wrap_WinterLights_UV_animation, parameterName=WNL_Animation_Panning_SetA_VehicleSide, defaultValue=<rgba 0 0 0 1>}
  n54 = MaterialFunctionCall(Float4=NamedRerouteUsage()) {materialFunction=&BreakOutFloat4Components}
  n55 = LinearInterpolate(A=@n51.RGB, B=LinearInterpolate(A=@n52.RGB, B=@n53.RGB, Alpha=@n54.G), Alpha=@n54.R)
  n56 = VectorParameter() {group=Wrap_WinterLights_UV_animation, parameterName=WNL_Animation_Panning_SetA_Mech, defaultValue=<rgba 0 0 0 1>}
  n57 = VectorParameter() {group=Wrap_WinterLights_UV_animation, parameterName=WNL_Animation_Panning_SetA_Character, defaultValue=<rgba 0 0 0 1>}
  n58 = LinearInterpolate(A=@n56.RGB, B=@n57.RGB, Alpha=@n54.A)
  n59 = ComponentMask(None=Multiply(A=Time(), B=LinearInterpolate(A=@n55, B=@n58, Alpha=@n54.B))) {g=true, r=true}
  n60 = Add(A=Multiply(A=@n49, B=Multiply(A=@n50)), B=@n59) {constB=0}
  n61 = NamedRerouteUsage()
  n62 = Constant2Vector() {g=1, r=-1}
  n63 = Add(A=Multiply(A=Multiply(A=@n50, B=@n62), B=@n49), B=@n59)
  n64 = LinearInterpolate(A=TextureSample(UVs=Add(A=@n48, B=@n60), Tex=@n61).A, B=TextureSample(UVs=Add(A=@n48, B=@n63), Tex=@n61).A, Alpha=@n36.G)
  n65 = ScalarParameter() {defaultValue=10, group=Wrap_WinterLights_Emissive, parameterName=WNL_Emissive Brightness_SetA}
  n66 = Multiply(A=@n45.RGB, B=Multiply(A=@n64, B=Multiply(A=@n64, B=@n65) {constB=10}))
  n67 = ScalarParameter() {defaultValue=1, group=Wrap_WinterLights_Emissive, parameterName=Activate SetB}
  n68 = NamedRerouteUsage()
  n69 = NamedRerouteUsage()
  n70 = Time()
  n71 = VectorParameter() {group=Wrap_WinterLights_UV_animation, parameterName=WNL_Animation_Panning_SetB_Weapon, defaultValue=<rgba -0.15 0 0 1>}
  n72 = VectorParameter() {group=Wrap_WinterLights_UV_animation, parameterName=WNL_Animation_Panning_SetB_Vehicle, defaultValue=<rgba 0 0 0 1>}
  n73 = VectorParameter() {group=Wrap_WinterLights_UV_animation, parameterName=WNL_Animation_Panning_SetB_VehicleSide, defaultValue=<rgba 0 0 0 1>}
  n74 = MaterialFunctionCall(Float4=NamedRerouteUsage()) {materialFunction=&BreakOutFloat4Components}
  n75 = VectorParameter() {group=Wrap_WinterLights_UV_animation, parameterName=WNL_Animation_Panning_SetB_Mech, defaultValue=<rgba 0 0 0 1>}
  n76 = VectorParameter() {group=Wrap_WinterLights_UV_animation, parameterName=WNL_Animation_Panning_SetB_Character, defaultValue=<rgba 0 0 0 1>}
  n77 = LinearInterpolate(
   A=LinearInterpolate(A=@n71.RGB, B=LinearInterpolate(A=@n72.RGB, B=@n73.RGB, Alpha=@n74.G), Alpha=@n74.R)
   B=LinearInterpolate(A=@n75.RGB, B=@n76.RGB, Alpha=@n74.A)
   Alpha=@n74.B
  )
  n78 = Frac(None=ComponentMask(None=Multiply(A=@n70, B=@n77)) {g=true, r=true})
  n79 = ScalarParameter() {defaultValue=15, group=Wrap_WinterLights_Emissive, parameterName=WNL_Animation_SetB_Jump}
  n80 = Divide(
   A=Ceil(None=Multiply(A=@n78, B=@n79) {constB=30})
   B=ScalarParameter() {defaultValue=5, group=Wrap_WinterLights_Emissive, parameterName=WNL_Animation_SetB_Steps}
  )
  n81 = Add(
   A=Multiply(
    A=Multiply(A=@n68, B=@n69)
    B=ScalarParameter() {defaultValue=3, group=Wrap_WinterLights_Emissive, parameterName=WNL_Animation_Light_tileRate}
   )
   B=@n80
  )
  n82 = TextureSample(UVs=Add(A=OneMinus(None=@n81), B=@n37) {constB=0}, Tex=NamedRerouteUsage())
  n83 = Multiply(
   A=@n45.RGB
   B=Multiply(A=Power(Base=@n82.A), B=ScalarParameter() {defaultValue=5, group=Wrap_WinterLights_Emissive, parameterName=WNL_Emissive Brightness SetB})
  )
  n84 = Multiply(A=@n36.R, B=ScalarParameter() {defaultValue=1, group=Wrap_WinterLights_Emissive, parameterName=WNL_Animation_SetC_WarpAmount})
  n85 = NamedRerouteUsage()
  n86 = NamedRerouteUsage()
  n87 = Add(A=Multiply(A=Multiply(A=@n85, B=@n86)) {constB=0.5}, B=@n59)
  n88 = Multiply(
   A=TextureSample(UVs=Add(A=@n84, B=@n87), Tex=NamedRerouteUsage()).A
   B=ScalarParameter() {defaultValue=10, group=Wrap_WinterLights_Emissive, parameterName=WNL_Emissive Brightness SetC}
  ) {constB=10}
  n89 = Add(
   A=Add(A=Multiply(A=@n47, B=@n66), B=Multiply(A=@n67, B=@n83))
   B=Multiply(A=ScalarParameter() {group=Wrap_WinterLights_Emissive, parameterName=Activate SetC}, B=Multiply(A=@n45.RGB, B=@n88))
  )
  n90 = Time()
  n91 = ScalarParameter() {defaultValue=-9999999, group=Wrap_WinterLights_Emissive, parameterName=WNL_Game_Time}
  n92 = Saturate(None=Divide(A=Subtract(A=@n90, B=@n91), B=ScalarParameter() {defaultValue=0.5, group=Wrap_WinterLights_Emissive, parameterName=WNL_Game_Time_Duration}))
  n93 = TextureObjectParameter() {
   group=Wrap_WinterLights
   parameterName=WNL_Main_Normal
   samplerType=SAMPLERTYPE_Normal
   sortPriority=10
   texture=&FlatNormalBlue
  }
  n94 = SetMaterialAttributes!(
   MaterialAttributes=@n41.MaterialAttributes
   BaseColor=@n44.RGB
   Roughness=LinearInterpolate(
    A=ScalarParameter() {group=Wrap_WinterLights, parameterName=WNL_Main_Roghness_Low, sortPriority=0}
    B=ScalarParameter() {defaultValue=1, group=Wrap_WinterLights, parameterName=WNL_Main_Roghness_High, sortPriority=0}
    Alpha=@n44.Alpha
   )
   Specular=ScalarParameter() {defaultValue=0.5, group=Wrap_WinterLights, parameterName=WNL_Main_Specular, sortPriority=0}
   EmissiveColor=Multiply(A=LinearInterpolate(A=@n46, B=@n89, Alpha=@n92), B=@n36.A)
   Metallic=ScalarParameter() {group=Wrap_WinterLights, parameterName=WNL_Main_Metallic, sortPriority=0}
   Normal=Add(
    A=@n41.Normal
    B=MaterialFunctionCall!(
     Texture_Normals=@n93
     Rotation_Axis=@n35
     Caculated_Bounds=@n31
     Pre_Skinned_Normals=@n6
     Blend_Contrast=@n7
     Fill_alternate_Axis_Opacity=@n8
     Is_Vehicle=@n3.R
     Is_Mech=@n3.R
    ) {materialFunction=&/MF_TopsSideProjection_Normals}.Add_Surface_Normal
   )
  ) {attributeSetTypes [
    #2
    #3
    #4
    #5
    #6
    #1
   ]}
  n95 = FunctionInput(Preview=@n39.R) {inputName=Mask Primary, inputType=FunctionInput_Scalar, sortPriority=24, previewValue=<1 0 0 1>}
  n96 = Saturate(None=Add(A=@n95, B=ScalarParameter() {group=Wrap_WinterLights, parameterName=WNL_Remove_Customization Mask, sortPriority=-10}))
  FunctionOutput(None=StaticSwitchParameter(True=BlendMaterialAttributes(A=@n41.MaterialAttributes, B=@n94, Alpha=@n96) {vertexAttributeBlendType=UseA}, False=@n40) {desc=TopSide Projection Base - Emissive pan Mask, group=Customization_Wrap, parameterName=Wrap_WinterLights})
  NamedRerouteDeclaration(None=ComponentMask(None=@n29) {r=true}) {name=TopSide_TileRate, nodeColor=<rgba 1 0.375 0 1>}
  n97 = TextureObjectParameter() {
   group=Wrap_WinterLights_Emissive
   parameterName=WNL_Animation_Gradent
   sortPriority=10
   texture=&T_BlacksRGB
  }
  NamedRerouteDeclaration(None=@n97) {name=LightTextureSample, nodeColor=<rgba 0 0.1328 1 1>}
  n98 = NamedRerouteUsage()
  n99 = MaterialFunctionCall(Color=NamedRerouteUsage()) {materialFunction=&SplitComponents}
  n100 = MaterialFunctionCall(Float4=NamedRerouteUsage()) {materialFunction=&BreakOutFloat4Components}
  n101 = LinearInterpolate(
   A=ScalarParameter() {group=Wrap_WinterLights_UV_animation, parameterName=WNL_Animation_Rotation_Weapon}
   B=LinearInterpolate(
    A=ScalarParameter() {group=Wrap_WinterLights_UV_animation, parameterName=WNL_Animation_Rotation_Vehicle}
    B=ScalarParameter() {group=Wrap_WinterLights_UV_animation, parameterName=WNL_Animation_Rotation_VehicleSide}
    Alpha=@n100.G
   )
   Alpha=@n100.R
  )
  n102 = LinearInterpolate(
   A=ScalarParameter() {defaultValue=0.25, group=Wrap_WinterLights_UV_animation, parameterName=WNL_Animation_Rotation_Mech}
   B=ScalarParameter() {group=Wrap_WinterLights_UV_animation, parameterName=WNL_Animation_Rotation_Character}
   Alpha=@n100.A
  )
  NamedRerouteDeclaration(None=MaterialFunctionCall(
    UVs=MaterialFunctionCall!(
     For_Safe_Zones
     Normal_Map
     Caculated_Bounds=@n98
     Pre_Skinned_Normals
     Is Vehicle=@n99.R
     Is VehicleSide=@n99.R
     Is Mech=@n99.R
     Vehicle_Rotation
     UV_Rotation Bypass
     Mech_Rotation
     Uvs_Offset
     Offset-after rotation
     Uvs_TileRate
    ) {materialFunction=&/MF_Projection_Side_Normal}.UVs No Offsets/Tiling
    Rotation Center=Constant2Vector() {g=0.5, r=0.5}
    "Rotation Angle (0-1)"=LinearInterpolate(A=@n101, B=@n102, Alpha=@n100.B)
   ) {materialFunction=&CustomRotator}.Rotated Values) {name=LightUv_Rotation, nodeColor=<rgba 1 0.3281 0 1>}
  NamedRerouteDeclaration(None=AppendVector(A=@n80, B=Constant())) {name=Bset_UVs, nodeColor=<rgba 0.8281 1 0 1>}
  NamedRerouteDeclaration(None=FunctionInput(Preview=Constant4Vector().RGBA) {bUsePreviewValueAsDefault=true, inputName=IsState, inputType=FunctionInput_Vector4, sortPriority=4}) {name=IsState, nodeColor=<rgba 0.1497 1 0.1633 1>}
  NamedRerouteDeclaration(None=FunctionInput(Preview=Constant2Vector()) {bUsePreviewValueAsDefault=true, inputName=IsState_B, inputType=FunctionInput_Vector2, sortPriority=4}) {name=IsState_B, nodeColor=<rgba 0.1497 1 0.1633 1>}
  n103 = PreSkinnedLocalBounds()
  n104 = Subtract(A=VertexInterpolator(VS=PreSkinnedPosition()).PS, B=@n103.Min)
  n105 = Subtract(A=@n103.Max, B=@n103.Max)
  NamedRerouteDeclaration(None=FunctionInput(Preview=Divide(A=@n104, B=@n105)) {bUsePreviewValueAsDefault=true, inputName=Calcated_Bounds, sortPriority=11}) {name=Caclcated_Bounds, nodeColor=<rgba 0 0.9297 1 1>}
  n106 = PreSkinnedNormal()
  NamedRerouteDeclaration(None=FunctionInput(Preview=VertexInterpolator(VS=@n106).PS) {bUsePreviewValueAsDefault=true, inputName=Pre-Skinned Local Normal, sortPriority=10}) {name=PreSkin_Normal, nodeColor=<rgba 0.8281 1 0 1>}
 ]}
