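# Material function for the WarriorC (Chicken Warrior) cosmetic wrap, blended over input material attributes through customization masks behind a static switch.
# Top/side projections sample two texture sets, with offset, scale and rotation selected for weapon, vehicle or mech.
# A height-lerped crack mask reveals the second, Fresnel-fuzz cloth layer, fading back over a reset duration after a Blueprint-supplied time.
# Gradient, subsurface color and secondary-blend branches do not feed the output.
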MaterialFunction /Game/Weapons/WeaponSkins/Wraps/MaterialFunctions/MF_Wrap_WarriorC
graph {nodes [
  n1 = FunctionInput() {inputName=Material, inputType=FunctionInput_MaterialAttributes}
  n2 = GetMaterialAttributes(None=@n1) {attributeGetTypes [#1]}
  n3 = ScalarParameter() {group=Wrap_ChickenWarrior, parameterName=CW_Specular_Min, sortPriority=-20}
  n4 = TextureObjectParameter() {group=Wrap__ChickenWarrior_Textures, parameterName=CW_Specular_01, sortPriority=20, texture=&T_BlacksRGB}
  n5 = FunctionInput(Preview=Constant()) {bUsePreviewValueAsDefault=true, inputName=IsState, inputType=FunctionInput_Vector4}
  n6 = MaterialFunctionCall(Color=@n5) {materialFunction=&SplitComponents}
  n7 = LinearInterpolate(
   A=ScalarParameter() {defaultValue=1, group=Wrap_ChickenWarrior, parameterName=CW_Blend_Contrast, sortPriority=-1}
   B=ScalarParameter() {defaultValue=2, group=Wrap_ChickenWarrior, parameterName=CW_Blend_Contrast_Mech, sortPriority=-1}
   Alpha=@n6.B
  )
  n8 = ScalarParameter() {group=Wrap_ChickenWarrior, parameterName=CW_Fill_alternate_Axis_Opacity, sortPriority=2}
  n9 = VectorParameter() {group=Wrap_ChickenWarrior, parameterName=CW_Fill_alternate_Axis_Color_Specular_01, sortPriority=21, defaultValue=<rgba 0.2 0.2 0.2 0.9>}
  n10 = VectorParameter() {group=Wrap_ChickenWarrior_UV, parameterName=CW_Offset_Weapon_01---------------------, sortPriority=10, defaultValue=<rgba 0 0 0 1>}
  n11 = VectorParameter() {group=Wrap_ChickenWarrior_UV, parameterName=CW_Offset_Vehicle_01, sortPriority=11, defaultValue=<rgba 0 0 0 1>}
  n12 = VectorParameter() {group=Wrap_ChickenWarrior_UV, parameterName=CW_Offset_Vehicle_Side_01, sortPriority=11, defaultValue=<rgba 0 0 0 1>}
  n13 = LinearInterpolate(A=@n10.RGB, B=LinearInterpolate(A=@n11.RGB, B=@n12.RGB, Alpha=@n6.G), Alpha=@n6.R)
  n14 = VectorParameter() {group=Wrap_ChickenWarrior_UV, parameterName=CW_Offset_Mech_01, sortPriority=12, defaultValue=<rgba 0 0 0 1>}
  n15 = PreSkinnedPosition()
  n16 = PreSkinnedLocalBounds()
  n17 = Divide(A=Subtract(A=VertexInterpolator(VS=@n15).PS, B=@n16.Min), B=Subtract(A=@n16.Max, B=@n16.Max))
  n18 = VectorParameter() {group=Wrap_ChickenWarrior_UV, parameterName="CW_Scale_Weapon_01 A=uniform", sortPriority=30, defaultValue=<rgba 1 1 1 1>}
  n19 = Multiply(A=@n18.RGB, B=@n18.RGB)
  n20 = VectorParameter() {group=Wrap_ChickenWarrior_UV, parameterName="CW_Scale_Vehicle_01 A=uniform", sortPriority=31, defaultValue=<rgba 1 1 1 1>}
  n21 = VectorParameter() {group=Wrap_ChickenWarrior_UV, parameterName="CW_Scale_Vehicle_Side_01 A=uniform", sortPriority=31, defaultValue=<rgba 1 1 1 1>}
  n22 = LinearInterpolate(A=Multiply(A=@n20.RGB, B=@n20.RGB), B=Multiply(A=@n21.RGB, B=@n21.RGB), Alpha=@n6.G)
  n23 = VectorParameter() {group=Wrap_ChickenWarrior_UV, parameterName="CW_Scale_Mech_01 A=uniform", defaultValue=<rgba 1 1 1 1>}
  n24 = Multiply(
   A=Add(
    A=FunctionInput(Preview=@n17) {bUsePreviewValueAsDefault=true, inputName=Calcated_Bounds, sortPriority=11}
    B=LinearInterpolate(A=@n13, B=@n14.RGB, Alpha=@n6.B)
   )
   B=LinearInterpolate(A=LinearInterpolate(A=@n19, B=@n22, Alpha=@n6.R), B=Multiply(A=@n23.RGB, B=@n23.RGB), Alpha=@n6.B)
  )
  n25 = FunctionInput(Preview=VertexInterpolator(VS=PreSkinnedNormal()).PS) {bUsePreviewValueAsDefault=true, inputName=Pre-Skinned Local Normal, sortPriority=10}
  n26 = VectorParameter() {group=Wrap_ChickenWarrior_UV, parameterName=CW-Rotation-Weapon-01---------------------, sortPriority=1, defaultValue=<rgba 0 0 -0.25 1>}
  n27 = VectorParameter() {group=Wrap_ChickenWarrior_UV, parameterName=CW_Rotation_Vehicle_01, sortPriority=2, defaultValue=<rgba 0 0 0 1>}
  n28 = VectorParameter() {group=Wrap_ChickenWarrior_UV, parameterName=CW_Rotation_Vehicle_Side_01, sortPriority=2, defaultValue=<rgba 0 0 0 1>}
  n29 = VectorParameter() {group=Wrap_ChickenWarrior_UV, parameterName=CW_Rotation_Mech_01, sortPriority=3, defaultValue=<rgba 0.25 0.25 0 1>}
  n30 = LinearInterpolate(A=LinearInterpolate(A=@n26.RGB, B=LinearInterpolate(A=@n27.RGB, B=@n28.RGB, Alpha=@n6.G), Alpha=@n6.R), B=@n29.RGB, Alpha=@n6.B)
  n31 = MaterialFunctionCall!(
   IsCharacter
   IsStatsB
   Texture=@n4
   Rotation_Axis=@n30
   Caculated_Bounds=@n24
   Pre_Skinned_Normals=@n25
   Blend_Contrast=@n7
   Fill_alternate_Axis_Opacity=@n8
   "Fill_alternate_Axis_Color_A=Roughness"=AppendVector(A=@n9.RGB, B=@n9.RGB)
   Is_Vehicle=@n6.R
   Is_Mech=@n6.B
   Enable_Fill_Z_Weapon
   Fill_Z_Weapon_opacity
   Fill_Z_Weapon_Color
  ) {desc=Specular, materialFunction=&/MF_TopsSideProjection_Base}
  n32 = ComponentMask(None=@n31.Combined) {a=true}
  n33 = ScalarParameter() {defaultValue=0.5, group=Wrap_ChickenWarrior_Cloth, parameterName=CW Cloth Scattering Min}
  n34 = ScalarParameter() {defaultValue=1, group=Wrap_ChickenWarrior_Cloth, parameterName=CW Cloth Scattering Max}
  n35 = TextureObjectParameter() {group=Wrap__ChickenWarrior_Textures, parameterName=CW_Mask_01, sortPriority=20, texture=&T_BlacksRGB}
  n36 = VectorParameter() {group=Wrap_ChickenWarrior, parameterName=CW_Fill_alternate_Axis_Color_Mask_01, sortPriority=21, defaultValue=<rgba 0 0 0 1>}
  n37 = MaterialFunctionCall(
   Texture=@n35
   Rotation_Axis=@n30
   Caculated_Bounds=@n24
   Pre_Skinned_Normals=@n25
   Blend_Contrast=@n7
   Fill_alternate_Axis_Opacity=@n8
   "Fill_alternate_Axis_Color_A=Roughness"=AppendVector(A=@n36.RGB, B=@n36.RGB)
   Is_Vehicle=@n6.R
   Is_Mech=@n6.B
  ) {desc=Masks, materialFunction=&MF_TopsSideProjection_Base}
  n38 = ComponentMask(None=@n37.Combined) {b=true}
  n39 = LinearInterpolate(
   A=ComponentMask(None=@n31.Combined) {r=true}
   B=Saturate(None=LinearInterpolate(A=@n33, B=@n34, Alpha=@n38))
   Alpha=ScalarParameter() {defaultValue=0.5, group=Wrap_ChickenWarrior_Cloth, parameterName=Cloth Specular Blend}
  )
  n40 = ChannelMaskParameter(None=@n37.Combined) {group=Wrap_ChickenWarrior, maskChannel=Green, parameterName=CW_CrackTransition_ChannelSelect, defaultValue=<rgba 0 1 0 0>}
  n41 = ScalarParameter() {group=Wrap_ChickenWarrior_Reactive, parameterName=CW_CrackTransitionMultiplier}
  n42 = ScalarParameter() {group=Wrap_ChickenWarrior_Reactive, parameterName=CW_CrackTransitionOverride}
  n43 = Subtract(A=Time(), B=ScalarParameter() {defaultValue=-99999998430674944, group=Wrap_ChickenWarrior_Reactive, parameterName=CW_TimeFromBP})
  n44 = ScalarParameter() {defaultValue=2, group=Wrap_ChickenWarrior_Reactive, parameterName=CW_ResetDuration}
  n45 = OneMinus(None=Saturate(None=Divide(A=@n43, B=@n44)))
  n46 = LinearInterpolate(B=LinearInterpolate(A=Multiply(A=@n40.R, B=@n41), B=@n42, Alpha=@n42), Alpha=@n45)
  n47 = ChannelMaskParameter(None=@n37.Combined) {group=Wrap_ChickenWarrior, parameterName=CW_CrackHeight_ChannelSelect}
  n48 = ScalarParameter() {defaultValue=1, group=Wrap_ChickenWarrior, parameterName=CW_Blend_Heightlerp_Contrast, sortPriority=71}
  n49 = LinearInterpolate(
   A=Constant()
   B=Saturate(None=MaterialFunctionCall(Transition Phase=@n46, Height Texture=@n47.R, Contrast=@n48) {materialFunction=&HeightLerp}.Alpha)
   Alpha=@n45
  )
  n50 = ScalarParameter() {defaultValue=1, group=Wrap_ChickenWarrior, parameterName=CW_Specular_Max, sortPriority=-20}
  n51 = ScalarParameter() {group=Wrap_ChickenWarrior, parameterName=CW_Roughness_Min, sortPriority=-20}
  n52 = TextureObjectParameter() {
   group=Wrap__ChickenWarrior_Textures
   parameterName=CW_BaseColor_01
   sortPriority=11
   texture=&T_BlacksRGB
  }
  n53 = VectorParameter() {group=Wrap_ChickenWarrior, parameterName="CW_Fill_alternate_Axis_Color_A=Roughness 01", sortPriority=3}
  n54 = MaterialFunctionCall(
   Texture=@n52
   Rotation_Axis=@n30
   Caculated_Bounds=@n24
   Pre_Skinned_Normals=@n25
   Blend_Contrast=@n7
   Fill_alternate_Axis_Opacity=@n8
   "Fill_alternate_Axis_Color_A=Roughness"=AppendVector(A=@n53.RGB, B=@n53.RGB)
   Is_Vehicle=@n6.R
   Is_Mech=@n6.B
  ) {desc=Base Color Rough, materialFunction=&MF_TopsSideProjection_Base}
  n55 = ComponentMask(None=@n54.Combined) {a=true}
  n56 = TextureObjectParameter() {
   group=Wrap__ChickenWarrior_Textures
   parameterName=CW_BaseColor_02
   sortPriority=11
   texture=&T_BlacksRGB
  }
  n57 = VectorParameter() {group=Wrap_ChickenWarrior, parameterName="CW_Fill_alternate_Axis_Color_A=Roughness 02", sortPriority=3}
  n58 = MaterialFunctionCall(
   Texture=@n56
   Rotation_Axis=@n30
   Caculated_Bounds=@n24
   Pre_Skinned_Normals=@n25
   Blend_Contrast=@n7
   Fill_alternate_Axis_Opacity=@n8
   "Fill_alternate_Axis_Color_A=Roughness"=AppendVector(A=@n57.RGB, B=@n57.RGB)
   Is_Vehicle=@n6.R
   Is_Mech=@n6.B
  ) {materialFunction=&MF_TopsSideProjection_Base}
  n59 = Saturate(None=Multiply(
    A=ComponentMask(None=@n58.Combined) {a=true}
    B=ScalarParameter() {defaultValue=1, group=Wrap_ChickenWarrior_Cloth, parameterName=CW Cloth Roughness}
   ))
  n60 = ScalarParameter() {defaultValue=1, group=Wrap_ChickenWarrior, parameterName=CW_Roughness_Max, sortPriority=-20}
  n61 = VectorParameter() {group=Wrap_ChickenWarrior_Cloth, parameterName=CW Cloth Fuzz Tint, defaultValue=<rgba 1 0.956 0.8714 1>}
  n62 = ScalarParameter() {defaultValue=1.5, group=Wrap_ChickenWarrior_Cloth, parameterName=CW Fuzz Exponent}
  n63 = TextureObjectParameter() {
   group=Wrap__ChickenWarrior_Textures
   parameterName=CW_Normal_02
   samplerType=SAMPLERTYPE_Normal
   sortPriority=16
   texture=&FlatNormalBlue
  }
  n64 = MaterialFunctionCall!(
   Texture_Normals=@n63
   Rotation_Axis=@n30
   Caculated_Bounds=@n24
   Pre_Skinned_Normals=@n25
   Blend_Contrast=@n7
   Fill_alternate_Axis_Opacity=@n8
   Is_Vehicle=@n6.R
   Is_Mech=@n6.B
  ) {materialFunction=&/MF_TopsSideProjection_Normals}
  n65 = Add(
   A=@n2.Normal
   B=Multiply(A=@n64.Add_Surface_Normal, B=ScalarParameter() {defaultValue=2, group=Wrap_ChickenWarrior_Cloth, parameterName=CW Fresnel Normal Intensity})
  )
  n66 = TextureCoordinate()
  n67 = ScalarParameter() {defaultValue=50, group=Wrap_ChickenWarrior_Cloth, parameterName=CW Fuzz Tiling}
  n68 = Multiply(A=Fresnel(ExponentIn=@n62, Normal=Transform(None=@n65)) {baseReflectFraction=0}, B=TextureSample(UVs=Multiply(A=@n66, B=@n67), Tex=@n35).A)
  n69 = ScalarParameter() {defaultValue=0.5, group=Wrap_ChickenWarrior_Cloth, parameterName=CW Fuzz Fresnel Blend}
  n70 = Multiply(A=Saturate(None=Multiply(A=@n68, B=@n69))) {constB=10}
  n71 = Multiply(
   A=Multiply(
    A=ComponentMask(None=@n58.Combined) {b=true, g=true, r=true}
    B=ScalarParameter() {defaultValue=0.8, group=Wrap_ChickenWarrior_Cloth, parameterName=CW Cloth Base Color Intensity}
   )
   B=Multiply(A=@n61.RGB, B=Add(A=@n70))
  )
  n72 = ScalarParameter() {group=Wrap_ChickenWarrior, parameterName=CW_Metallic, sortPriority=-20}
  n73 = TextureObjectParameter() {
   group=Wrap__ChickenWarrior_Textures
   parameterName=CW_Normal_01
   samplerType=SAMPLERTYPE_Normal
   sortPriority=16
   texture=&FlatNormalBlue
  }
  n74 = MaterialFunctionCall(
   Texture_Normals=@n73
   Rotation_Axis=@n30
   Caculated_Bounds=@n24
   Pre_Skinned_Normals=@n25
   Blend_Contrast=@n7
   Fill_alternate_Axis_Opacity=@n8
   Is_Vehicle=@n6.R
   Is_Mech=@n6.B
  ) {materialFunction=&MF_TopsSideProjection_Normals}
  n75 = LinearInterpolate(A=@n74.Add_Surface_Normal, B=@n64.Add_Surface_Normal, Alpha=@n49)
  n76 = ScalarParameter() {defaultValue=1, group=Wrap_ChickenWarrior, parameterName=CW_NormalIntensity, sortPriority=-20}
  n77 = SetMaterialAttributes!(
   MaterialAttributes=@n2.MaterialAttributes
   BaseColor=LinearInterpolate(A=ComponentMask(None=@n54.Combined) {b=true, g=true, r=true}, B=@n71, Alpha=@n49)
   Roughness=LinearInterpolate(A=@n51, B=@n60, Alpha=LinearInterpolate(A=@n55, B=@n59, Alpha=@n49))
   Normal=LinearInterpolate(A=@n2.Normal, B=Add(A=@n2.Normal, B=@n75), Alpha=@n76)
   Specular=LinearInterpolate(A=@n3, B=@n50, Alpha=LinearInterpolate(A=@n32, B=@n39, Alpha=@n49))
   Metallic=@n72
  ) {attributeSetTypes [
    #2
    #3
    #1
    #4
    #5
   ]}
  n78 = FunctionInput() {bUsePreviewValueAsDefault=true, inputName="Customization Mask (_CM B Top left)", inputType=FunctionInput_Scalar, sortPriority=24}
  n79 = FunctionInput() {bUsePreviewValueAsDefault=true, inputName="Customization Mask 2 (_CM B Top Right)", inputType=FunctionInput_Scalar, sortPriority=25}
  n80 = Add(A=@n78, B=@n79)
  n81 = Subtract(A=NamedRerouteUsage(), B=NamedRerouteUsage())
  n82 = BlendMaterialAttributes(A=@n2.MaterialAttributes, B=@n77, Alpha=LinearInterpolate(A=Saturate(None=@n80), B=@n79, Alpha=Saturate(None=@n81))) {vertexAttributeBlendType=UseA}
  n83 = SetMaterialAttributes(
   MaterialAttributes=@n2.MaterialAttributes
   BaseColor=ComponentMask(None=@n54.Combined) {b=true, g=true, r=true}
   Roughness=LinearInterpolate(A=@n51, B=@n60, Alpha=ComponentMask(None=@n54.Combined) {a=true})
   Normal=LinearInterpolate(A=@n2.Normal, B=Add(A=@n2.Normal, B=@n74.Add_Surface_Normal), Alpha=@n76)
   Specular=LinearInterpolate(A=@n3, B=@n50, Alpha=@n32)
   Metallic=@n72
  ) {attributeSetTypes [
    #2
    #3
    #1
    #4
    #5
   ]}
  BlendMaterialAttributes(
   A=@n82
   B=@n83
   Alpha=LinearInterpolate(A=@n78, B=Constant(), Alpha=ScalarParameter() {group=Wrap_ChickenWarrior, parameterName=CW_FORWRAPROLL_DontUseSecondary, sortPriority=-500})
  )
  n84 = DotProduct(A=PixelNormalWS(), B=CameraVectorWS())
  Clamp(None=Power(Base=OneMinus(None=@n84), Exp=@n62))
  n85 = VertexInterpolator(VS=PreSkinnedPosition())
  n86 = VectorParameter() {group=Wrap_ChickenWarrior_Gradient, parameterName=CW_Weapon_GradientCenterOffset, defaultValue=<rgba 0 0 0 1>}
  n87 = VectorParameter() {group=Wrap_ChickenWarrior_Gradient, parameterName=CW_Vehicle_GradientCenterOffset, defaultValue=<rgba 0 0 0 1>}
  n88 = MaterialFunctionCall(Color=@n5) {materialFunction=&SplitComponents}
  n89 = LinearInterpolate(A=@n86.RGB, B=@n87.RGB, Alpha=@n88.R)
  n90 = VectorParameter() {group=Wrap_ChickenWarrior_Gradient, parameterName=CW_Mech_GradientCenterOffset, defaultValue=<rgba 0 0 0 1>}
  n91 = PreSkinnedLocalBounds()
  n92 = Divide(A=Add(A=@n85.PS, B=LinearInterpolate(A=@n89, B=@n90.RGB, Alpha=@n88.B)), B=@n91.Max)
  n93 = ChannelMaskParameter(None=@n92) {group=Wrap_ChickenWarrior_Gradient, maskChannel=Green, parameterName=CW_WeaponGradientDirection, defaultValue=<rgba 0 1 0 0>}
  n94 = ChannelMaskParameter(None=@n92) {group=Wrap_ChickenWarrior_Gradient, maskChannel=Green, parameterName=CW_VehicleGradientDirection, defaultValue=<rgba 0 1 0 0>}
  n95 = ChannelMaskParameter(None=@n92) {group=Wrap_ChickenWarrior_Gradient, maskChannel=Blue, parameterName=CW_MechGradientDirection, defaultValue=<rgba 0 0 1 0>}
  Clamp(None=LinearInterpolate(A=LinearInterpolate(A=@n93.R, B=@n94.R), B=@n95.R))
  Constant()
  FunctionOutput(None=StaticSwitchParameter(True=@n82, False=@n1) {desc="TopSide Projection Base ", group=Customization_Wrap, parameterName=Wrap_WarriorC})
  n96 = VectorParameter() {group=Wrap_ChickenWarrior, parameterName=CW SSS Chicken Shadows, defaultValue=<rgba 0.4226 0.3248 0.1207 1>}
  n97 = VectorParameter() {group=Wrap_ChickenWarrior, parameterName=CW SSS Chicken Color, defaultValue=<rgba 0.7605 0.6584 0.4452 1>}
  LinearInterpolate(
   A=VectorParameter() {group=Wrap_ChickenWarrior, parameterName=CW SSS Base, defaultValue=<rgba 0 0 0 1>}.RGB
   B=LinearInterpolate(A=@n96.RGB, B=@n97.RGB, Alpha=@n38)
   Alpha=@n49
  )
  MaterialFunctionCall!("Customization Mask (_CM B Top left)"=@n78, "Customization Mask 2 (_CM B Top Right)"=@n79) {materialFunction=&/MF_Customization_AllMasking}
  NamedRerouteDeclaration(None=MaterialFunctionCall(Float4=@n5) {materialFunction=&BreakOutFloat4Components}.A) {name=IsCharacter, nodeColor=<rgba 0.8536 1 0.0771 1>}
  NamedRerouteDeclaration(None=FunctionInput() {bUsePreviewValueAsDefault=true, inputName=isCharAccent, inputType=FunctionInput_Scalar, sortPriority=78}) {name=isCharAccent, nodeColor=<rgba 0 0.9253 1 1>}
  n98 = Divide(A=Subtract(A=@n85.PS, B=@n91.Min), B=Subtract(A=@n91.Max, B=@n91.Max))
  n99 = Constant3Vector() {constant=<rgba -0.5 -0.5 -0.5 0>}
  n100 = VectorParameter() {group=Wrap_ChickenWarrior_Gradient, parameterName=CW_TransitionPosition_Offset, defaultValue=<rgba 0 0 0 1>}
  OneMinus(None=SphereMask(
    A=Add(A=@n98, B=@n99)
    B=@n100.RGB
    Radius=ScalarParameter() {defaultValue=1, group=Wrap_ChickenWarrior_Gradient, parameterName=CW_ShatterRadius Multiplier}
    Hardness=ScalarParameter() {group=Wrap_ChickenWarrior_Gradient, parameterName=CW_ShatterGradientHardness}
   ))
  OneMinus(None=Multiply(A=ScalarParameter() {defaultValue=2, group=Wrap_ChickenWarrior_Reactive, parameterName=IB_BaseGrowth}, B=@n45))
  Saturate(None=@n71)
 ]}
