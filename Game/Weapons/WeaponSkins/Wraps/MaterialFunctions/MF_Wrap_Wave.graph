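# Material function for the animated "Wave" wrap skin, blended onto an input material through primary and secondary masks behind the Wrap_Wave static switch.
# Side-projected UVs, with rotation chosen by item-state inputs (weapon, vehicle, mech, character, backpack, pickaxe), feed three time-panned, noise-warped bar layers (A, B, C).
# The base material lerps two warped curve-atlas color ramps, plus roughness and specular, by combined masks.
# A separate flat-color accent material covers the secondary mask.
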MaterialFunction /Game/Weapons/WeaponSkins/Wraps/MaterialFunctions/MF_Wrap_Wave
graph {nodes [
  n1 = TextureSample() {samplerType=SAMPLERTYPE_Masks, texture=/Game/Weapons/FORT_Rifles/Textures/T_Scar_CM}
  n2 = FunctionInput(Preview=@n1.R) {bUsePreviewValueAsDefault=true, inputName=Mask Primary, inputType=FunctionInput_Scalar, sortPriority=24, previewValue=<1 0 0 1>}
  n3 = FunctionInput(Preview=@n1.B) {bUsePreviewValueAsDefault=true, inputName=Mask Secondary, inputType=FunctionInput_Scalar, sortPriority=24, previewValue=<1 0 0 1>}
  n4 = BlendMaterialAttributes(A=NamedRerouteUsage(), B=NamedRerouteUsage(), Alpha=Saturate(None=Add(A=@n2, B=@n3))) {vertexAttributeBlendType=UseA}
  n5 = NamedRerouteUsage()
  n6 = NamedRerouteDeclaration(None=@n3) {name=Msk_Accent, nodeColor=<rgba 0 1 0.7656 1>}
  n7 = TextureSample() {samplerType=SAMPLERTYPE_Normal, texture=/Game/Weapons/FORT_Rifles/T_SCAR_N}
  n8 = Constant3Vector() {constant=<rgba 1 0 0 0>}
  n9 = GetMaterialAttributes(None=FunctionInput(Preview=SetMaterialAttributes!(MaterialAttributes, Normal=@n7.RGB, BaseColor=@n8) {attributeSetTypes [#1, #2]}) {inputName=Material, inputType=FunctionInput_MaterialAttributes}) {attributeGetTypes [
    #2
    #3
    #1
    #4
   ]}
  FunctionOutput(None=StaticSwitchParameter(
    True=BlendMaterialAttributes(A=@n4, B=@n5, Alpha=@n6) {vertexAttributeBlendType=UseA}
    False=NamedRerouteDeclaration(None=@n9.MaterialAttributes) {name=Root_Material, nodeColor=<rgba 0.7812 1 0 1>}
   ) {group=Customization_Wrap, parameterName=Wrap_Wave})
  n10 = ComponentMask(None=Multiply(A=Time(), B=VectorParameter() {group=Wrap_Wave_BarsB, parameterName=Wave_BarsB_Breaks_Speed, defaultValue=<rgba -0.05 0.1 0 1>}.RGB)) {g=true, r=true}
  n11 = Multiply(
   A=ScalarParameter() {defaultValue=2, group=Wrap_Wave_BarsB, parameterName=Wave_BarsB_Breaks_TileRate}
   B=MaterialFunctionCall(
    UVs=NamedRerouteUsage()
    Rotation Center=Constant() {r=0.5}
    "Rotation Angle (0-1)"=ScalarParameter() {defaultValue=-0.07, group=Wrap_Wave_BarsB, parameterName=Wave_BarsB_Break_Rotation}
   ) {materialFunction=&CustomRotator}.Rotated Values
  )
  n12 = NamedRerouteUsage()
  n13 = ScalarParameter() {defaultValue=0.5, group=Wrap_Wave_BarsB, parameterName=Wave_WarpB_TileRate}
  n14 = Multiply(A=Time(), B=VectorParameter() {group=Wrap_Wave_BarsB, parameterName=Wave_WarpB_Speed, defaultValue=<rgba 0.0233 -0.013 0 1>}.RGB)
  n15 = TextureSample(UVs=Add(A=Multiply(A=@n12, B=@n13), B=ComponentMask(None=@n14) {g=true, r=true}), Tex=NamedRerouteUsage()) {samplerSource=SSM_Wrap_WorldGroupSettings}
  n16 = ScalarParameter() {defaultValue=0.25, group=Wrap_Wave_BarsB, parameterName=Wave_BarsB_Breaks_Strength}
  n17 = TextureSample(UVs=Add(A=Add(A=@n10, B=@n11), B=Multiply(A=@n15.R, B=@n16) {constB=0.5}), Tex=NamedRerouteUsage()) {samplerSource=SSM_Wrap_WorldGroupSettings}
  Multiply(A=@n17.A)
  Multiply(A=MaterialFunctionCall(Value_Morning=Constant() {r=0.5}, Value_Day=Constant() {r=1}, Value_Evening=Constant() {r=0.5}, Value_Night=Constant() {r=0.25}) {materialFunction=&Scalar_Time_of_Day}.Result)
  n18 = VectorParameter() {group=Wrap_Wave Accent, parameterName=Wave_Accent_BaseColor, defaultValue=<rgba 0.0074 0.0175 0.0573 1>}
  n19 = SetMaterialAttributes!(
   MaterialAttributes=NamedRerouteUsage()
   BaseColor=@n18.RGB
   Roughness=ScalarParameter() {defaultValue=0.5, group=Wrap_Wave Accent, parameterName=Wave_Accent_Roughness}
   Specular=ScalarParameter() {defaultValue=0.25, group=Wrap_Wave Accent, parameterName=Wave_Accent_Spec}
   Metallic=ScalarParameter() {group=Wrap_Wave Accent, parameterName=Wave_Accent_Metalic}
   Normal
   EmissiveColor
  ) {attributeSetTypes [
    #2
    #3
    #5
    #4
    #1
    #6
   ]}
  NamedRerouteDeclaration(None=@n19) {name=Mat_Accent, nodeColor=<rgba 0 0.3906 1 1>}
  n20 = FunctionInput(Preview=@n1.G) {bUsePreviewValueAsDefault=true, inputName=Mask Crevise, inputType=FunctionInput_Scalar, sortPriority=24, previewValue=<1 0 0 1>}
  NamedRerouteDeclaration(None=@n20) {name=Msk_Curvature, nodeColor=<rgba 0 1 0.7422 1>}
  n21 = NamedRerouteUsage()
  n22 = NamedRerouteUsage()
  n23 = ScalarParameter() {defaultValue=1.5, group=Wrap_Wave_BarsA, parameterName=Wave_WarpA_TileRate}
  n24 = Multiply(A=Time(), B=VectorParameter() {group=Wrap_Wave_BarsA, parameterName=Wave_WarpA_Speed, defaultValue=<rgba 0.05 0.013 0 1>}.RGB)
  n25 = TextureSample(UVs=Add(A=Multiply(A=@n22, B=@n23), B=ComponentMask(None=@n24) {g=true, r=true}), Tex=NamedRerouteUsage()) {samplerSource=SSM_Wrap_WorldGroupSettings}
  n26 = Add(
   A=Add(A=ComponentMask(None=@n21) {r=true}, B=Multiply(A=@n25.R) {constB=0.25})
   B=ScalarParameter() {defaultValue=-0.6, group=Wrap_Wave_BarsA, parameterName=Wave_BarsA_MSk_Offset}
  ) {constB=-0.6}
  n27 = ScalarParameter() {defaultValue=0.1, group=Wrap_Wave_BarsA, parameterName=Wave_BarsA_MSk_Falloff}
  NamedRerouteDeclaration(None=Saturate(None=Divide(A=@n26, B=@n27) {constB=0.1})) {name=MSk_right, nodeColor=<rgba 0.1484 1 0 1>}
  n28 = Divide(A=@n26, B=Multiply(A=@n27) {constB=-1}) {constB=-0.1}
  n29 = ScalarParameter() {defaultValue=6, group=Wrap_Wave_BarsA, parameterName=Wave_BarsA_TileRate}
  n30 = Multiply(A=Time(), B=VectorParameter() {group=Wrap_Wave_BarsA, parameterName=Wave_BarsA_Speed, defaultValue=<rgba 0.05 0 0 1>}.RGB)
  n31 = Add(
   A=Add(A=Multiply(A=@n21, B=@n29), B=ComponentMask(None=@n30) {g=true, r=true})
   B=Multiply(A=ScalarParameter() {defaultValue=0.4, group=Wrap_Wave_BarsA, parameterName=Wave_WarpA_Strength}, B=@n25.R)
  )
  n32 = NamedRerouteUsage()
  NamedRerouteDeclaration(None=Multiply(A=Saturate(None=@n28), B=TextureSample(UVs=@n31, Tex=@n32) {samplerSource=SSM_Wrap_WorldGroupSettings}.R)) {name=Msk_BarsA, nodeColor=<rgba 0.5469 0 1 1>}
  n33 = NamedRerouteUsage()
  n34 = Constant() {r=0.5}
  n35 = ScalarParameter() {defaultValue=-0.15, group=Wrap_Wave_BarsB, parameterName=Wave_BarsB_Rotation}
  n36 = Time()
  n37 = VectorParameter() {group=Wrap_Wave_BarsB, parameterName=Wave_BarsB_Speed, defaultValue=<rgba -0.05 0 0 1>}
  n38 = Add(
   A=Multiply(
    A=MaterialFunctionCall(UVs=@n33, Rotation Center=@n34, "Rotation Angle (0-1)"=@n35) {materialFunction=&CustomRotator}.Rotated Values
    B=ScalarParameter() {defaultValue=2, group=Wrap_Wave_BarsB, parameterName=Wave_BarsB_TileRate}
   )
   B=ComponentMask(None=Multiply(A=@n36, B=@n37.RGB)) {g=true, r=true}
  )
  n39 = Multiply(A=ScalarParameter() {defaultValue=0.3, group=Wrap_Wave_BarsB, parameterName=Wave_WarpB_Strength}, B=@n15.R)
  n40 = Subtract(A=TextureSample(UVs=Add(A=@n38, B=@n39), Tex=NamedRerouteUsage()) {samplerSource=SSM_Wrap_WorldGroupSettings}.R, B=@n17.A)
  n41 = ComponentMask(None=@n33) {g=true}
  n42 = Multiply(A=@n15.R) {constB=0.15}
  n43 = Divide(
   A=Add(A=Add(A=@n41, B=@n42), B=ScalarParameter() {defaultValue=-0.65, group=Wrap_Wave_BarsB, parameterName=Wave_BarsB_ZOffset}) {constB=-0.8}
   B=ScalarParameter() {defaultValue=-0.4, group=Wrap_Wave_BarsB, parameterName=Wave_BarsB_Zfalloff}
  ) {constB=-1}
  n44 = Add(A=Multiply(A=Saturate(None=@n40), B=Saturate(None=@n43)), B=NamedRerouteUsage())
  NamedRerouteDeclaration(None=Multiply(A=NamedRerouteUsage(), B=Saturate(None=@n44))) {name="Msk_BarsB\\C", nodeColor=<rgba 0.5469 0 1 1>}
  n45 = NamedRerouteUsage()
  n46 = ScalarParameter() {defaultValue=0.3, group=Wrap_Wave_BarsC, parameterName=Wave_WarpC_TileRate}
  n47 = Multiply(A=Time(), B=VectorParameter() {group=Wrap_Wave_BarsC, parameterName=Wave_WarpC_Speed, defaultValue=<rgba 0.0233 0.013 0 1>}.RGB)
  n48 = TextureSample(UVs=Add(A=Multiply(A=@n45, B=@n46), B=ComponentMask(None=@n47) {g=true, r=true}), Tex=NamedRerouteUsage()) {samplerSource=SSM_Wrap_WorldGroupSettings}
  n49 = Multiply(A=@n48.R) {constB=0.05}
  n50 = Add(A=ComponentMask(None=NamedRerouteUsage()) {g=true}, B=@n49)
  n51 = ScalarParameter() {defaultValue=-0.45, group=Wrap_Wave_BarsC, parameterName=Wave_BarsC_ZOffset}
  n52 = Saturate(None=Divide(A=Add(A=@n50, B=@n51) {constB=-0.4}, B=ScalarParameter() {defaultValue=0.15, group=Wrap_Wave_BarsC, parameterName=Wave_BarsC_Zfalloff}) {constB=0.25})
  n53 = MaterialFunctionCall(
   UVs=NamedRerouteUsage()
   Rotation Center=Constant() {r=0.5}
   "Rotation Angle (0-1)"=ScalarParameter() {defaultValue=0.05, group=Wrap_Wave_BarsC, parameterName=Wave_BarsC_Rotation}
  ) {materialFunction=&CustomRotator}
  n54 = Add(
   A=Multiply(A=@n53.Rotated Values, B=ScalarParameter() {defaultValue=2.5, group=Wrap_Wave_BarsC, parameterName=Wave_BarsC_TileRate})
   B=Multiply(A=ScalarParameter() {defaultValue=0.75, group=Wrap_Wave_BarsC, parameterName=Wave_WarpC_Strength}, B=@n48.R)
  )
  n55 = ComponentMask(None=Multiply(A=Time(), B=VectorParameter() {group=Wrap_Wave_BarsC, parameterName=Wave_BarsC_Speed, defaultValue=<rgba -0.1 0.3 0 1>}.RGB)) {g=true, r=true}
  n56 = Add(
   A=ComponentMask(None=Add(A=@n53.Rotated Values, B=@n49)) {r=true}
   B=ScalarParameter() {defaultValue=-0.75, group=Wrap_Wave_BarsC, parameterName=Wave_BarsC_Offset}
  ) {constB=-0.8}
  n57 = ScalarParameter() {defaultValue=-0.1, group=Wrap_Wave_BarsC, parameterName=Wave_BarsC_falloff}
  n58 = Multiply(
   A=TextureSample(UVs=Add(A=@n54, B=@n55), Tex=NamedRerouteUsage()) {samplerSource=SSM_Wrap_WorldGroupSettings}.R
   B=Saturate(None=Divide(A=@n56, B=@n57) {constB=-0.05})
  )
  NamedRerouteDeclaration(None=Multiply(A=Multiply(A=@n52, B=@n58))) {name=Msk BarsC, nodeColor=<rgba 0.9688 0 1 1>}
  n59 = ScalarParameter() {group=Wrap_Wave_Uvs, parameterName=Wave_Rotation_Vehicle, sortPriority=150}
  n60 = ScalarParameter() {group=Wrap_Wave_Uvs, parameterName=Wave_Rotation_Vehicle_Side, sortPriority=150}
  n61 = MaterialFunctionCall(Float4=NamedRerouteUsage()) {materialFunction=&BreakOutFloat4Components}
  n62 = LinearInterpolate(
   A=LinearInterpolate(
    A=ScalarParameter() {group=Wrap_Wave_Uvs, parameterName=Wave_Rotation_Weapon, sortPriority=150}
    B=LinearInterpolate(A=@n59, B=@n60, Alpha=@n61.G)
    Alpha=@n61.R
   )
   B=LinearInterpolate(
    A=ScalarParameter() {group=Wrap_Wave_Uvs, parameterName=Wave_Rotation_Mech, sortPriority=150}
    B=ScalarParameter() {group=Wrap_Wave_Uvs, parameterName=Wave_Rotation_Character, sortPriority=150}
    Alpha=@n61.A
   )
   Alpha=@n61.B
  )
  n63 = ScalarParameter() {group=Wrap_Wave_Uvs, parameterName=Wave_Rotation_Backpack, sortPriority=150}
  n64 = MaterialFunctionCall(Float2=NamedRerouteUsage()) {materialFunction=&BreakOutFloat2Components}
  NamedRerouteDeclaration(None=LinearInterpolate(
    A=LinearInterpolate(A=@n62, B=@n63, Alpha=@n64.R)
    B=ScalarParameter() {group=Wrap_Wave_Uvs, parameterName=Wave_Rotation_Pickaxe, sortPriority=150}
    Alpha=@n64.G
   )) {name=UVs_Rotation, nodeColor=<rgba 0 0.7188 1 1>}
  n65 = ScalarParameter() {defaultValue=88, group=Wrap_Wave, parameterName="Wave_CCBars_B  Index"}
  n66 = NamedRerouteUsage()
  n67 = ScalarParameter() {defaultValue=0.75, group=Wrap_Wave_BarsA, parameterName=Wave_Color_WarpA_TileRate}
  n68 = Multiply(A=Time(), B=VectorParameter() {group=Wrap_Wave_BarsA, parameterName=Wave_Color_WarpA_Speed, defaultValue=<rgba 0.05 0.013 0 1>}.RGB)
  n69 = TextureSample(UVs=Add(A=Multiply(A=@n66, B=@n67), B=ComponentMask(None=@n68) {g=true, r=true}), Tex=NamedRerouteUsage()) {samplerSource=SSM_Wrap_WorldGroupSettings}
  n70 = ScalarParameter() {defaultValue=0.2, group=Wrap_Wave_BarsA, parameterName=Wave_Color_WarpA_Amount}
  n71 = Multiply(A=@n70) {constB=-1}
  n72 = Add(A=ComponentMask(None=NamedRerouteUsage()) {r=true}, B=Multiply(A=Add(A=@n69.R, B=@n71) {constB=-0.5}, B=@n70) {constB=0})
  n73 = CurveAtlasRowParameter(CurveTime=@n72) {
   atlas=/Game/Weapons/WeaponSkins/Wraps/Textures/CruveAtlas_Wraps_01
   curve=/Game/Weapons/WeaponSkins/Wraps/Textures/Wave/CC_WaveB
   defaultValue=88
   group=Wrap_Wave
   parameterName=Wave_CCBars_B
  }
  n74 = ScalarParameter() {defaultValue=87, group=Wrap_Wave, parameterName="Wave_CCBars  Index"}
  n75 = CurveAtlasRowParameter(CurveTime=@n72) {
   atlas=&CruveAtlas_Wraps_01
   curve=/Game/Weapons/WeaponSkins/Wraps/Textures/Wave/CC_WaveA
   defaultValue=87
   group=Wrap_Wave
   parameterName=Wave_CCBars
  }
  n76 = Saturate(None=Add(A=NamedRerouteUsage(), B=NamedRerouteUsage()))
  n77 = SetMaterialAttributes(
   MaterialAttributes=NamedRerouteUsage()
   BaseColor=LinearInterpolate(
    A=MaterialFunctionCall!(Index=@n65, CCinput=@n73, CCinput_Alpha, UV=@n72) {materialFunction=&/MF_Global_CC_Tex}.3V
    B=MaterialFunctionCall(Index=@n74, CCinput=@n75, UV=@n72) {materialFunction=&MF_Global_CC_Tex}.3V
    Alpha=@n76
   )
   Roughness=LinearInterpolate(
    A=ScalarParameter() {defaultValue=0.45, group=Wrap_Wave, parameterName=Wave_Roughness_High}
    B=ScalarParameter() {defaultValue=0.25, group=Wrap_Wave, parameterName=Wave_Roughness_Low}
    Alpha=@n76
   )
   Specular=LinearInterpolate(
    A=ScalarParameter() {defaultValue=0.4, group=Wrap_Wave, parameterName=Wave_Spec_High}
    B=ScalarParameter() {defaultValue=0.5, group=Wrap_Wave, parameterName=Wave_Spec_Low}
    Alpha=@n76
   )
   Metallic=ScalarParameter() {group=Wrap_Wave, parameterName=Wave_Metalic}
  ) {attributeSetTypes [
    #2
    #3
    #5
    #4
    #1
    #6
   ]}
  NamedRerouteDeclaration(None=@n77) {name=Mat_Base, nodeColor=<rgba 0 0.3906 1 1>}
  n78 = NamedRerouteUsage()
  n79 = NamedRerouteUsage()
  n80 = MaterialFunctionCall(Float4=NamedRerouteUsage()) {materialFunction=&BreakOutFloat4Components}
  NamedRerouteDeclaration(None=MaterialFunctionCall(
    UVs=MaterialFunctionCall!(
     For_Safe_Zones
     Normal_Map
     Caculated_Bounds=@n78
     Pre_Skinned_Normals=@n79
     Is Vehicle=@n80.R
     Is VehicleSide=@n80.R
     Is Mech=@n80.R
     Vehicle_Rotation
     UV_Rotation Bypass
     Mech_Rotation
     Uvs_Offset
     Offset-after rotation
     Uvs_TileRate
    ) {materialFunction=&/MF_Projection_Side_Normal}.UVs
    Rotation Center=Constant() {r=0.5}
    "Rotation Angle (0-1)"=NamedRerouteUsage()
   ) {materialFunction=&CustomRotator}.Rotated Values) {name=UVs_Side, nodeColor=<rgba 0 1 0.0625 1>}
  NamedRerouteDeclaration(None=FunctionInput(Preview=Constant4Vector().RGBA) {bUsePreviewValueAsDefault=true, inputName=IsState, inputType=FunctionInput_Vector4, sortPriority=4}) {name=IsState, nodeColor=<rgba 0.1497 1 0.1633 1>}
  NamedRerouteDeclaration(None=FunctionInput(Preview=Constant2Vector()) {bUsePreviewValueAsDefault=true, inputName=IsState_B, inputType=FunctionInput_Vector2, sortPriority=4}) {name=IsState_B, nodeColor=<rgba 0.1497 1 0.1633 1>}
  n81 = PreSkinnedLocalBounds()
  n82 = Subtract(A=VertexInterpolator(VS=PreSkinnedPosition()).PS, B=@n81.Min)
  n83 = MaterialFunctionCall(Color=Subtract(A=@n81.Max, B=@n81.Max)) {materialFunction=&SplitComponents}
  n84 = Max(A=Max(A=@n83.R, B=@n83.R), B=@n83.B)
  NamedRerouteDeclaration(None=FunctionInput(Preview=Divide(A=@n82, B=@n84)) {bUsePreviewValueAsDefault=true, inputName=Calcated_Bounds, sortPriority=11}) {name=Caclcated_Bounds, nodeColor=<rgba 0 0.9297 1 1>}
  NamedRerouteDeclaration(None=TextureObjectParameter() {group=Wrap_Wave, parameterName=Wave_Tex__MskA, texture=&T_BlacksRGB}) {name=Tex_MskA, nodeColor=<rgba 1 0 0.75 1>}
  n85 = PreSkinnedNormal()
  NamedRerouteDeclaration(None=FunctionInput(Preview=VertexInterpolator(VS=@n85).PS) {bUsePreviewValueAsDefault=true, inputName=Pre-Skinned Local Normal, sortPriority=10}) {name=PreSkin_Normal, nodeColor=<rgba 0.8281 1 0 1>}
  NamedRerouteDeclaration(None=TextureObjectParameter() {group=Wrap_Wave, parameterName=Wave_Tex__MskB, texture=&T_BlacksRGB}) {name=Tex_MskB, nodeColor=<rgba 0 0.7891 1 1>}
  NamedRerouteUsage()
 ]}
