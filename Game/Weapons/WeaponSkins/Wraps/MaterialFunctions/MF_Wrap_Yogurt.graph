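# Material function for the Yogurt weapon-skin wrap: an animated flower, line and dot pattern for weapons, vehicles and mechs.
# An isState vector picks per-target offsets, tiling and axis masks for side-normal projected UVs warped by panning textures.
# Scrolling sphere-mask curve-atlas gradients color the layers, with world-aligned aura, fresnel, pulsing glow and time-of-day emissive.
# The Wrap_Yogurt static switch blends the result over incoming material attributes through customization masks.
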MaterialFunction /Game/Weapons/WeaponSkins/Wraps/MaterialFunctions/MF_Wrap_Yogurt
graph {nodes [
  n1 = FunctionInput(Preview=Constant()) {inputName=isState, inputType=FunctionInput_Vector4, sortPriority=5}
  n2 = MaterialFunctionCall(Color=@n1) {materialFunction=&SplitComponents}
  n3 = ScalarParameter() {group=Wrap_Yogurt_UVs, parameterName=Yogurt_Rotation_Vehicle}
  n4 = ScalarParameter() {defaultValue=0.25, group=Wrap_Yogurt_UVs, parameterName=Yogurt_Rotation_Mech -----}
  n5 = VectorParameter() {group=Wrap_Yogurt_UVs, parameterName=Yogurt_02_Offset, defaultValue=<rgba -0.46 0 0 1>}
  n6 = VectorParameter() {desc=must activate TS_Vehicle_Side_01 UVs, group=Wrap_Yogurt_UVs, parameterName=Yogurt_01_Offset_Vehicle_Side, defaultValue=<rgba 0 0 0 1>}
  n7 = MaterialFunctionCall(Color=@n1) {materialFunction=&SplitComponents}
  n8 = LinearInterpolate(
   A=VectorParameter() {group=Wrap_Yogurt_UVs, parameterName=Yogurt_01_Offset_Weapon, defaultValue=<rgba 0 0 0 1>}.RGB
   B=LinearInterpolate(A=VectorParameter() {group=Wrap_Yogurt_UVs, parameterName=Yogurt_01_Offset_Vehicle, defaultValue=<rgba 0 0 0 1>}.RGB, B=@n6.RGB, Alpha=@n7.G)
   Alpha=@n7.R
  )
  n9 = VectorParameter() {group=Wrap_Yogurt_UVs, parameterName=Yogurt_01_Offset_Mech -----, defaultValue=<rgba 0 0 0 1>}
  n10 = Multiply(A=@n24.UVs No Offsets/Tiling, B=ScalarParameter() {defaultValue=1, group=Wrap_Yogurt_UVs, parameterName=Yogurt_01_Warp_TileRate})
  n11 = MaterialFunctionCall(Color=@n1) {materialFunction=&SplitComponents}
  n12 = LinearInterpolate(
   A=VectorParameter() {group=Wrap_Yogurt_UVs, parameterName=Yogurt_01_Warp_Speed_Weapon, defaultValue=<rgba 0.1 0.1 0 1>}.RGB
   B=LinearInterpolate(
    A=VectorParameter() {group=Wrap_Yogurt_UVs, parameterName=Yogurt_01_Warp_Speed_Vehicle, defaultValue=<rgba 0.1 0.025 0 1>}.RGB
    B=VectorParameter() {group=Wrap_Yogurt_UVs, parameterName=Yogurt_01_Warp_Speed_VehicleSide, defaultValue=<rgba 0.1 0.1 0 1>}.RGB
    Alpha=@n11.G
   )
   Alpha=@n11.R
  )
  n13 = VectorParameter() {group=Wrap_Yogurt_UVs, parameterName=Yogurt_01_Warp_Speed_Mech, defaultValue=<rgba 0.1 0.1 0 1>}
  n14 = ComponentMask(None=Multiply(A=Time(), B=LinearInterpolate(A=@n12, B=@n13.RGB, Alpha=@n11.B))) {g=true, r=true}
  n15 = TextureObjectParameter() {group=Wrap__Yogurt_Textures, parameterName=Yogurt_MaskMap_02, texture=&T_BlacksRGB}
  n16 = ChannelMaskParameter(None=TextureSample(UVs=Add(A=@n10, B=@n14), Tex=@n15) {samplerType=SAMPLERTYPE_LinearGrayscale}.RGBA) {group=Wrap_Yogurt_UVs, maskChannel=Alpha, parameterName=Yogurt_01_Warp_Channel, defaultValue=<rgba 0 0 0 1>}
  n17 = ScalarParameter() {defaultValue=0.01, group=Wrap_Yogurt_UVs, parameterName=Yogurt_01_Warp_Amount -----}
  n18 = ScalarParameter() {defaultValue=1, group=Wrap_Yogurt_UVs, parameterName=Yogurt_01_TileRate_Weapon}
  n19 = LinearInterpolate(
   A=ScalarParameter() {defaultValue=1, group=Wrap_Yogurt_UVs, parameterName=Yogurt_01_TileRate_Vehicle}
   B=ScalarParameter() {defaultValue=1, group=Wrap_Yogurt_UVs, parameterName=Yogurt_01_TileRate_VehicleSide}
   Alpha=@n7.G
  )
  n20 = FunctionInput(Preview=VertexInterpolator(VS=PreSkinnedNormal()).PS) {bUsePreviewValueAsDefault=true, inputName=Pre-Skinned Local Normal, sortPriority=10}
  n21 = VertexInterpolator(VS=PreSkinnedPosition())
  n22 = PreSkinnedLocalBounds()
  n23 = FunctionInput(Preview=Divide(A=Subtract(A=@n21.PS, B=@n22.Min), B=Subtract(A=@n22.Max, B=@n22.Max))) {bUsePreviewValueAsDefault=true, inputName=Calcated_Bounds, sortPriority=11}
  n24 = MaterialFunctionCall!(
   For_Safe_Zones
   Normal_Map
   Caculated_Bounds=@n23
   Pre_Skinned_Normals=@n20
   Is Vehicle=@n2.R
   Is VehicleSide=@n2.R
   Is Mech=@n2.R
   Vehicle_Rotation=@n3
   UV_Rotation Bypass
   Mech_Rotation=@n4
   Uvs_Offset=Add(A=LinearInterpolate(A=@n8, B=@n9.RGB, Alpha=@n7.B), B=Multiply(A=@n16.R, B=@n17))
   Offset-after rotation
   Uvs_TileRate=LinearInterpolate(
    A=LinearInterpolate(A=@n18, B=@n19, Alpha=@n7.R)
    B=ScalarParameter() {defaultValue=1, group=Wrap_Yogurt_UVs, parameterName=Yogurt_01_TileRate_Mech -----}
    Alpha=@n7.B
   )
  ) {materialFunction=&/MF_Projection_Side_Normal}
  n25 = Multiply(A=@n24.UVs No Offsets/Tiling, B=ScalarParameter() {defaultValue=1, group=Wrap_Yogurt_UVs, parameterName=Yogurt_02_Warp_TileRate})
  n26 = ChannelMaskParameter(None=TextureSample(UVs=Add(A=@n25, B=@n14), Tex=@n15) {samplerType=SAMPLERTYPE_LinearGrayscale}.RGBA) {group=Wrap_Yogurt_UVs, maskChannel=Alpha, parameterName=Yogurt_02_Warp_Channel, defaultValue=<rgba 0 0 0 1>}
  n27 = Multiply(A=@n26.R, B=ScalarParameter() {defaultValue=0.025, group=Wrap_Yogurt_UVs, parameterName=Yogurt_02_Warp_Amount -----})
  n28 = ScalarParameter() {defaultValue=1, group=Wrap_Yogurt, parameterName=Yogurt_CurveRadial_Size_Weapon, sortPriority=100}
  n29 = MaterialFunctionCall(Color=@n1) {materialFunction=&SplitComponents}
  n30 = LinearInterpolate(
   A=ScalarParameter() {defaultValue=1, group=Wrap_Yogurt, parameterName=Yogurt_CurveRadial_Size_Vehicle, sortPriority=100}
   B=ScalarParameter() {defaultValue=1, group=Wrap_Yogurt, parameterName=Yogurt_CurveRadial_Size_VehicleSide, sortPriority=100}
   Alpha=@n29.G
  )
  n31 = SphereMask(
   A=MaterialFunctionCall(
    Caculated_Bounds=@n23
    Pre_Skinned_Normals=@n20
    Is Vehicle=@n2.R
    Is VehicleSide=@n2.R
    Is Mech=@n2.R
    Vehicle_Rotation=@n3
    Mech_Rotation=@n4
    Uvs_Offset=Add(A=@n5.RGB, B=@n27)
    Uvs_TileRate=Constant() {r=1}
   ) {materialFunction=&MF_Projection_Side_Normal}.UVs
   B=Constant() {r=0.5}
   Radius=LinearInterpolate(
    A=LinearInterpolate(A=@n28, B=@n30, Alpha=@n29.R)
    B=ScalarParameter() {defaultValue=1, group=Wrap_Yogurt, parameterName=Yogurt_CurveRadial_Size_Mech, sortPriority=100}
    Alpha=@n29.B
   )
   Hardness=Constant()
  ) {hardnessPercent=0}
  n32 = Multiply(A=ScalarParameter() {defaultValue=1, group=Wrap_Yogurt, parameterName=Yogurt_CurveRadial_TileRate, sortPriority=100}, B=@n31)
  n33 = Multiply(A=Time(), B=ScalarParameter() {defaultValue=0.33, group=Wrap_Yogurt, parameterName=Yogurt_CurveRadial_Speed, sortPriority=100})
  n34 = Frac(None=Add(A=@n32, B=@n33))
  n35 = CurveAtlasRowParameter(CurveTime=@n34) {
   atlas=/Game/Weapons/WeaponSkins/Wraps/Textures/CruveAtlas_Wraps_01
   curve=/Game/Weapons/WeaponSkins/Wraps/Textures/Yogurt/C_Yogurt_Grad
   defaultValue=48
   group=Wrap__Yogurt_Textures
   parameterName=Yogurt_Curve_Grad
   sortPriority=3
  }
  n36 = MaterialFunctionCall!(
   Index=ScalarParameter() {defaultValue=48, group=Wrap__Yogurt_Textures, parameterName="Yogurt_Curve_Grad  Index", sortPriority=3}
   CCinput=@n35
   CCinput_Alpha=@n35
   UV=@n34
  ) {materialFunction=&/MF_Global_CC_Tex}
  n37 = Multiply(
   A=@n36.3V
   B=Multiply(A=@n36.Alpha, B=ScalarParameter() {defaultValue=1, group=Wrap_Yogurt, parameterName=Yogurt_CurveRadial_CurveBrightness, sortPriority=100})
  )
  n38 = VectorParameter() {group=Wrap_Yogurt, parameterName="Yogurt_Aura_Tint - A=BlendPercentage", defaultValue=<rgba 1 1 1 0.1>}
  n39 = LinearInterpolate(A=@n37, B=@n38.RGB, Alpha=@n38.RGB)
  n40 = Add(A=Multiply(A=Time(), B=ScalarParameter() {defaultValue=0.075, group=Wrap_Yogurt, parameterName=Yogurt_Aura_WarpGradientspeed}), B=@n23)
  n41 = ScalarParameter() {defaultValue=1, group=Wrap_Yogurt, parameterName=Yogurt_Aura_WarpGradientTileRate}
  n42 = MaterialFunctionCall(Color=Constant3Vector() {constant=<rgba 0 -0.25 -0.25 1>}) {materialFunction=&SplitComponents}
  n43 = ChannelMaskParameter(None=MaterialFunctionCall(
    Export Float 4=StaticBool() {value=true}
    TextureObject=@n15
    TextureSize=Constant() {r=1}
    WorldPosition=Multiply(A=@n40, B=@n41)
    World Space Normal=@n20
    RotZAxis=@n42.B
    RotXAxis=@n42.B
    RotYAxis=@n42.B
    ProjectionTransitionContrast=ScalarParameter() {defaultValue=1, group=Wrap_Yogurt, parameterName=Yogurt_Aura_WarpBlendContrast}
   ) {materialFunction=&/WorldAlignedTexture}.XYZ Texture) {group=Wrap_Yogurt, parameterName=Yogurt_Aura_WarpGradientMaskChannel}
  n44 = ChannelMaskParameter(None=TextureSample(UVs=@n34, Tex=@n15) {desc=Mask 01}.RGBA) {group=Wrap_Yogurt, maskChannel=Green, parameterName=Yogurt_CurveHighlight_MaskChannel, sortPriority=100, defaultValue=<rgba 0 1 0 0>}
  n45 = Multiply(A=@n43.R, B=@n44.R)
  n46 = Multiply(A=@n45, B=Add(A=@n45, B=@n44.R))
  n47 = Multiply(A=@n39, B=Multiply(A=ScalarParameter() {defaultValue=1, group=Wrap_Yogurt, parameterName=Yogurt_Aura_Intensity-----}, B=Saturate(None=@n46)))
  n48 = MaterialFunctionCall(Color=Abs(None=@n20)) {materialFunction=&SplitComponents}
  n49 = Add(A=@n48.R, B=@n48.R)
  n50 = Add(A=@n48.B, B=@n48.B)
  n51 = MaterialFunctionCall(Color=@n1) {materialFunction=&SplitComponents}
  n52 = ScalarParameter() {defaultValue=0.1, group=Wrap_Yogurt, parameterName=Yogurt_Mask_Contrast, sortPriority=51}
  n53 = MaterialFunctionCall(In=LinearInterpolate(A=LinearInterpolate(A=@n49, B=@n50, Alpha=@n51.R), B=Add(A=@n48.R, B=@n48.R), Alpha=@n51.B), Contrast=@n52) {materialFunction=&CheapContrast}
  n54 = LinearInterpolate(
   A=Constant()
   B=Saturate(None=OneMinus(None=@n53.Result))
   Alpha=ScalarParameter() {defaultValue=1, group=Wrap_Yogurt, parameterName=Yogurt_Mask_AltAxisActive -----, sortPriority=50}
  )
  n55 = Frac(None=Add(A=@n32, B=@n33))
  n56 = CurveAtlasRowParameter(CurveTime=@n55) {
   atlas=&CruveAtlas_Wraps_01
   curve=/Game/Weapons/WeaponSkins/Wraps/Textures/Yogurt/C_Yogurt_BC
   defaultValue=47
   group=Wrap__Yogurt_Textures
   parameterName=Yogurt_Curve_BC
   sortPriority=3
  }
  n57 = MaterialFunctionCall(
   Index=ScalarParameter() {defaultValue=47, group=Wrap__Yogurt_Textures, parameterName="Yogurt_Curve_BC  Index", sortPriority=3}
   CCinput=@n56
   CCinput_Alpha=@n56
   UV=@n55
  ) {materialFunction=&MF_Global_CC_Tex}
  n58 = Multiply(
   A=@n57.3V
   B=Multiply(A=@n57.Alpha, B=ScalarParameter() {defaultValue=1, group=Wrap_Yogurt, parameterName=Yogurt-CurveBaseColor_Brightness-----, sortPriority=100})
  )
  n59 = LinearInterpolate(A=@n47, B=@n58, Alpha=@n54) {constB=0}
  n60 = Multiply(A=@n31, B=ScalarParameter() {defaultValue=1, group=Wrap_Yogurt, parameterName=Yogurt_Lines_Tilerate})
  n61 = Multiply(A=Time(), B=ScalarParameter() {defaultValue=0.15, group=Wrap_Yogurt, parameterName=Yogurt_Lines_Speed})
  n62 = TextureObjectParameter() {group=Wrap__Yogurt_Textures, parameterName=Yogurt_MaskMap_01, texture=&T_BlacksRGB}
  n63 = TextureSample(UVs=Frac(None=Add(A=@n60, B=@n61)), Tex=@n62)
  n64 = ChannelMaskParameter(None=@n63.RGBA) {group=Wrap_Yogurt, maskChannel=Blue, parameterName=Yogurt_Lines_WarpChannel, defaultValue=<rgba 0 0 1 0>}
  n65 = ChannelMaskParameter(None=@n63.RGBA) {group=Wrap_Yogurt, maskChannel=Blue, parameterName=Yogurt_Lines_MaskChannel, defaultValue=<rgba 0 0 1 0>}
  n66 = Multiply(A=@n64.R, B=@n65.R)
  n67 = Multiply(A=@n66, B=Add(A=@n66, B=@n65.R))
  n68 = ScalarParameter() {defaultValue=10, group=Wrap_Yogurt, parameterName=Yogurt_Dots_Intensity-----}
  n69 = VectorParameter() {group=Wrap_Yogurt_UVs, parameterName=Yogurt_Dots_Offset_Vehicle, defaultValue=<rgba 0 0 0 1>}
  n70 = VectorParameter() {group=Wrap_Yogurt_UVs, parameterName=Yogurt_Dots_Offset_Vehicle_Side, defaultValue=<rgba 0 0 0 1>}
  n71 = MaterialFunctionCall(Color=@n1) {materialFunction=&SplitComponents}
  n72 = LinearInterpolate(
   A=LinearInterpolate(
    A=VectorParameter() {group=Wrap_Yogurt_UVs, parameterName=Yogurt_Dots_Offset_Weapon, defaultValue=<rgba 0 0 0 1>}.RGB
    B=LinearInterpolate(A=@n69.RGB, B=@n70.RGB, Alpha=@n71.G)
    Alpha=@n71.R
   )
   B=VectorParameter() {group=Wrap_Yogurt_UVs, parameterName=Yogurt_Dots_Offset_Mech -----, defaultValue=<rgba 0 0 0 1>}.RGB
   Alpha=@n71.B
  )
  n73 = MaterialFunctionCall(Vector or UVs=Add(A=@n24.UVs No Offsets/Tiling, B=ComponentMask(None=@n72) {g=true, r=true})) {materialFunction=&VectorToRadialValue}
  n74 = VectorParameter() {group=Wrap_Yogurt, parameterName=Yogurt_Dots_WarpGradientspeed, defaultValue=<rgba -0.05 -0.25 0 1>}
  n75 = Add(A=@n73.Radial Coordinates, B=Multiply(A=Time(), B=ComponentMask(None=@n74.RGB) {g=true, r=true}))
  n76 = ScalarParameter() {defaultValue=2, group=Wrap_Yogurt, parameterName=Yogurt_Dots_WarpGradientTileRate}
  n77 = ChannelMaskParameter(None=TextureSample(UVs=Multiply(A=@n75, B=@n76), Tex=@n15).RGBA) {group=Wrap_Yogurt, maskChannel=Blue, parameterName=Yogurt_Dots_WarpMaskChannel, defaultValue=<rgba 0 0 1 0>}
  n78 = VectorParameter() {group=Wrap_Yogurt, parameterName=Yogurt_Dots_WarpGradientspeed, defaultValue=<rgba -0.05 -0.25 0 1>}
  n79 = Add(A=@n73.Radial Coordinates, B=Multiply(A=Time(), B=ComponentMask(None=@n78.RGB) {g=true, r=true}))
  n80 = ScalarParameter() {defaultValue=1, group=Wrap_Yogurt, parameterName=Yogurt_Dots_MaskTileRate}
  n81 = ChannelMaskParameter(None=TextureSample(UVs=Multiply(A=@n79, B=@n80), Tex=@n15).RGBA) {group=Wrap_Yogurt, parameterName=Yogurt_Dots_MaskChannel}
  n82 = Multiply(A=@n77.R, B=@n81.R)
  n83 = Saturate(None=Multiply(A=@n82, B=Add(A=@n82, B=@n81.R)))
  n84 = Constant()
  n85 = MaterialFunctionCall(Color=Abs(None=@n20)) {materialFunction=&SplitComponents}
  n86 = LinearInterpolate(A=LinearInterpolate(A=@n85.R, B=LinearInterpolate(A=@n85.B, B=@n85.B, Alpha=@n51.G), Alpha=@n51.R), B=@n85.G, Alpha=@n51.B)
  n87 = Saturate(None=OneMinus(None=MaterialFunctionCall(In=@n86, Contrast=@n52) {materialFunction=&CheapContrast}.Result))
  n88 = Constant() {r=1}
  n89 = Add(
   A=Multiply(A=ScalarParameter() {defaultValue=1, group=Wrap_Yogurt, parameterName=Yogurt_Lines_Intensity-----}, B=Saturate(None=@n67))
   B=LinearInterpolate(A=Multiply(A=@n68, B=@n83), Alpha=LinearInterpolate(A=@n84, B=@n87, Alpha=@n88)) {constB=0}
  )
  n90 = LinearInterpolate(A=@n58, B=@n37, Alpha=@n89)
  n91 = ChannelMaskParameter(None=TextureSample(UVs=@n24.UVs, Tex=@n62) {desc=Mask 01}.RGBA) {group=Wrap_Yogurt, maskChannel=Green, parameterName=Yogurt_FlowerStems_MaskChannel, defaultValue=<rgba 0 1 0 0>}
  n92 = TextureSample(UVs=@n24.UVs, Tex=@n62) {desc=Mask 02}
  n93 = ChannelMaskParameter(None=@n92.RGBA) {group=Wrap_Yogurt, maskChannel=Alpha, parameterName=Yogurt_FlowerGlow_MaskChannel, defaultValue=<rgba 0 0 0 1>}
  n94 = Multiply(A=@n93.R, B=@n43.R) {desc=+ outerr flower glow}
  n95 = LinearInterpolate(A=Add(A=Multiply(A=@n59) {constB=0.25}, B=LinearInterpolate(A=@n90, B=@n37, Alpha=@n91.R)) {desc="= Aura Glow"}, B=@n37, Alpha=@n94)
  n96 = Add(A=@n95, B=@n39)
  n97 = Multiply(A=@n44.R, B=ScalarParameter() {defaultValue=20, group=Wrap_Yogurt, parameterName=Yogurt_CurveHighlight_Brightness, sortPriority=100})
  n98 = LinearInterpolate(A=@n89, B=@n91.R, Alpha=@n91.R) {desc=Dots + Lines Mask}
  n99 = TextureSample(
   UVs=@n24.UVs
   Tex=TextureObjectParameter() {group=Wrap__Yogurt_Textures, parameterName=Yogurt_Flower, texture=&T_BlacksRGB}
  )
  n100 = ChannelMaskParameter(None=@n99.RGBA) {group=Wrap_Yogurt, maskChannel=Alpha, parameterName=Yogurt_Flower_MaskChannel-----, defaultValue=<rgba 0 0 0 1>}
  n101 = LinearInterpolate(A=@n98, B=@n100.R, Alpha=@n100.R)
  n102 = Multiply(A=@n97, B=@n101)
  n103 = LinearInterpolate(
   A=LinearInterpolate(A=LinearInterpolate(A=@n95, B=@n96, Alpha=@n102) {desc=+ Highlight GLows}, B=@n58, Alpha=@n54) {desc=Axis Masking}
   B=@n99.RGB
   Alpha=@n100.R
  ) {desc=+ Flower}
  n104 = LinearInterpolate(A=ChannelMaskParameter(None=@n92.RGBA) {group=Wrap_Yogurt, parameterName=Yogurt_FlowerInnerGlow_MaskChannel}.R, Alpha=@n54) {constB=0, desc=+ Inner Flower Glow}
  n105 = LinearInterpolate(A=@n101, Alpha=@n54) {constB=0}
  n106 = Multiply(A=ScalarParameter() {defaultValue=1, group=Wrap_Yogurt_Emissive, parameterName=Yogurt_Emissive_DarkenBaseColor, sortPriority=-20}, B=@n105)
  n107 = Multiply(
   A=LinearInterpolate(A=LinearInterpolate(A=@n103, B=@n37, Alpha=@n91.R), B=@n39, Alpha=@n104) {desc=+ Inner Glow}
   B=OneMinus(None=Saturate(None=@n106))
  )
  n108 = GetMaterialAttributes(None=FunctionInput() {inputName=Material, inputType=FunctionInput_MaterialAttributes, sortPriority=-20}) {attributeGetTypes [#1]}
  n109 = Multiply(
   A=Fresnel(
    ExponentIn=ScalarParameter() {defaultValue=8, group=Wrap_Yogurt, parameterName=Yogurt_Fresnel, sortPriority=1}
    Normal=Transform(None=@n108.Normal)
   ) {baseReflectFraction=0}
   B=ScalarParameter() {defaultValue=2, group=Wrap_Yogurt, parameterName=Yogurt_FresnelIntensity, sortPriority=1}
  )
  n110 = Add(A=@n109, B=Multiply(A=@n109, B=@n43.R))
  n111 = LinearInterpolate(A=@n107, B=@n39, Alpha=@n110)
  n112 = Add(A=@n110, B=@n105)
  n113 = LinearInterpolate(A=Add(A=Multiply(A=@n47) {constB=0.25}, B=Multiply(A=@n37, B=@n98)) {desc=+ Aura Glow}, B=@n37, Alpha=@n94)
  n114 = Add(A=@n97, B=@n113)
  n115 = LinearInterpolate(
   A=LinearInterpolate(A=LinearInterpolate(A=@n113, B=@n114, Alpha=@n101) {desc=+ Hightlight glows}, Alpha=@n54) {constB=0, desc=Axis masking}
   B=@n99.RGB
   Alpha=@n100.R
  )
  n116 = Sine(None=Multiply(A=Time(), B=ScalarParameter() {defaultValue=0.5, group=Wrap_Yogurt, parameterName=Yogurt_FlowerInnerGlow_PulseTime}) {constB=0.1}) {period=0.5}
  n117 = Clamp(
   None=Divide(A=Add(A=@n116))
   Min=ScalarParameter() {defaultValue=0.5, group=Wrap_Yogurt, parameterName=Yogurt_FlowerInnerGlow_PulseMin}
   Max=ScalarParameter() {defaultValue=1, group=Wrap_Yogurt, parameterName=Yogurt_FlowerInnerGlow_PulseMax}
  )
  n118 = ScalarParameter() {defaultValue=2.5, group=Wrap_Yogurt, parameterName=Yogurt_FlowerInnerGlow_Amount}
  n119 = LinearInterpolate(
   A=@n115
   B=Multiply(A=Add(A=@n115, B=@n39), B=Multiply(A=@n117, B=@n118))
   Alpha=Multiply(A=ScalarParameter() {defaultValue=10, group=Wrap_Yogurt_Emissive, parameterName=Yogurt_FlowerInnerGlow_Intensity, sortPriority=30}, B=@n104)
  ) {desc=+ Inner glow}
  n120 = ScalarParameter() {defaultValue=0.75, group=Wrap_Yogurt_Emissive, parameterName=Yogurt_Emissive_TOD_DawnDusk-----}
  n121 = MaterialFunctionCall(
   Value_Morning=@n120
   Value_Day=ScalarParameter() {defaultValue=1, group=Wrap_Yogurt_Emissive, parameterName=Yogurt_Emissive_TOD_Day}
   Value_Evening=@n120
   Value_Night=ScalarParameter() {defaultValue=0.5, group=Wrap_Yogurt_Emissive, parameterName=Yogurt_Emissive_TOD_Night}
  ) {materialFunction=&Scalar_Time_of_Day}
  n122 = SetMaterialAttributes!(
   MaterialAttributes=@n108.MaterialAttributes
   BaseColor=LinearInterpolate(A=@n107, B=Add(A=@n111, B=@n39), Alpha=@n110)
   Metallic=ScalarParameter() {group=Wrap_Yogurt, parameterName=Yogurt_Metallic, sortPriority=1}
   Specular=LinearInterpolate(
    A=ScalarParameter() {defaultValue=1, group=Wrap_Yogurt, parameterName=Yogurt_Specular_Min, sortPriority=1}
    B=ScalarParameter() {defaultValue=0.5, group=Wrap_Yogurt, parameterName=Yogurt_Specular_Max, sortPriority=1}
    Alpha=@n112
   )
   Roughness=LinearInterpolate(
    A=ScalarParameter() {defaultValue=0.5, group=Wrap_Yogurt, parameterName=Yogurt_Roughness, sortPriority=1}
    B=ScalarParameter() {defaultValue=1, group=Wrap_Yogurt, parameterName=Yogurt_Roughness_High, sortPriority=1}
    Alpha=@n112
   )
   EmissiveColor=LinearInterpolate(A=Multiply(A=@n119, B=@n121.Result), B=@n39, Alpha=@n110)
  ) {attributeSetTypes [
    #2
    #3
    #4
    #5
    #6
   ]}
  n123 = FunctionInput() {
   bUsePreviewValueAsDefault=true
   desc=base customization
   inputName="Customization Mask (_CM B Top left)"
   inputType=FunctionInput_Scalar
   sortPriority=49
  }
  n124 = FunctionInput() {
   bUsePreviewValueAsDefault=true
   desc=secondary
   inputName="Customization Mask 2 (_CM B Top Right)"
   inputType=FunctionInput_Scalar
   sortPriority=50
  }
  n125 = NamedRerouteUsage()
  n126 = NamedRerouteUsage()
  n127 = LinearInterpolate(A=Saturate(None=Add(A=@n123, B=@n124)), B=@n124, Alpha=Saturate(None=Subtract(A=@n125, B=@n126)))
  FunctionOutput(None=StaticSwitchParameter(
    True=BlendMaterialAttributes(A=@n108.MaterialAttributes, B=@n122, Alpha=@n127) {vertexAttributeBlendType=UseA}
    False=@n108.MaterialAttributes
   ) {group=Customization_Wrap, parameterName=Wrap_Yogurt}) {outputName=Result_1}
  NamedRerouteDeclaration(None=MaterialFunctionCall(Float4=@n1) {materialFunction=&BreakOutFloat4Components}.A) {name=IsCharacter, nodeColor=<rgba 0.8866 1 0 1>}
  NamedRerouteDeclaration(None=FunctionInput() {bUsePreviewValueAsDefault=true, inputName=isCharAccent, inputType=FunctionInput_Scalar, sortPriority=78}) {name=isCharAccent, nodeColor=<rgba 0 0.9253 1 1>}
 ]}
