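# Material function for the Zero Point weapon wrap, layering an animated, glowing Nexus-floor pattern over an input material within the customization mask, behind a static switch.
# Panning triplanar textures use pre-skinned position and normal, sized by local bounds, with an isVehicle input altering XY pan.
# Panning noise height-lerps against the pattern, driving normal, darkening, specular and emissive.
# Fresnel on the blended normal tints base color and adds chromatic emissive sheen.
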MaterialFunction /Game/Weapons/WeaponSkins/Wraps/MaterialFunctions/MF_ZeroPointWrap
graph {nodes [
  n1 = ScalarParameter() {defaultValue=0.01, parameterName=Pan_ZeroPoint_Mesh_y}
  n2 = ScalarParameter() {defaultValue=0.01, parameterName=Pan_ZeroPoint_Mesh_z}
  AppendVector(A=@n1, B=@n2)
  Constant() {r=1}
  n3 = VectorParameter() {parameterName=BaseColor, defaultValue=<rgba 0.2006 0.2822 0.4531 1>}
  n4 = VectorParameter() {parameterName=BaseColor2, defaultValue=<rgba 0.9792 0.9884 0.9948 1>}
  n5 = FunctionInput(Preview=MakeMaterialAttributes!(
    BaseColor
    Metallic
    Specular
    Roughness
    Anisotropy
    EmissiveColor
    Opacity
    OpacityMask
    Normal
    Tangent
    WorldPositionOffset
    SubsurfaceColor
    ClearCoat
    ClearCoatRoughness
    AmbientOcclusion
    Refraction
    CustomizedUVs_0
    CustomizedUVs_1
    CustomizedUVs_2
    CustomizedUVs_3
    CustomizedUVs_4
    CustomizedUVs_5
    CustomizedUVs_6
    CustomizedUVs_7
    PixelDepthOffset
    ShadingModel
    Displacement
   )) {bUsePreviewValueAsDefault=true, inputName=Material, inputType=FunctionInput_MaterialAttributes, sortPriority=-200}
  n6 = TextureObject() {samplerType=SAMPLERTYPE_Normal, texture=/Game/Athena/Environments/Nexus/Textures/NexFloor_N}
  n7 = MaterialFunctionCall(Float3=PreSkinnedLocalBounds().Half Extents) {materialFunction=&BreakOutFloat3Components}
  n8 = Divide(
   A=Max(A=Max(A=@n7.R, B=@n7.R), B=@n7.B)
   B=MaterialFunctionCall() {materialFunction=/Engine/Functions/Engine_MaterialFunctions02/WorldPositionOffset/ObjectScale}.Scale Z
  )
  n9 = Multiply(A=ScalarParameter() {defaultValue=1, parameterName=Size_ZeroPoint_Mesh}, B=@n8)
  n10 = VertexInterpolator(VS=PreSkinnedNormal())
  n11 = VertexInterpolator(VS=PreSkinnedPosition())
  n12 = AppendVector(A=@n1, B=@n2)
  n13 = ScalarParameter() {defaultValue=-0.0001, parameterName=Pan_ZeroPoint_Mesh_x}
  n14 = AppendVector(A=@n13, B=@n2)
  n15 = LinearInterpolate(
   A=AppendVector(A=@n13, B=@n1)
   B=AppendVector(A=Multiply(A=@n1) {constB=-1}, B=@n13)
   Alpha=FunctionInput() {inputName=isVehicle, inputType=FunctionInput_Scalar, sortPriority=10}
  )
  n16 = TextureObject() {texture=/Game/Global/Textures/Default/Noise/T_Tiling_Noise_01}
  n17 = Multiply(A=ScalarParameter() {defaultValue=1.3606, parameterName=Size_ZeroPoint_Noise}, B=@n8)
  n18 = ScalarParameter() {defaultValue=-0.05, parameterName=Pan_Noise_Mesh_y}
  n19 = ScalarParameter() {defaultValue=0.07, parameterName=Pan_Noise_Mesh_z}
  n20 = AppendVector(A=@n18, B=@n19)
  n21 = ScalarParameter() {defaultValue=-0.01, parameterName=Pan_Noise_Mesh_x}
  n22 = AppendVector(A=@n21, B=@n19)
  n23 = AppendVector(A=@n21, B=@n18)
  n24 = MaterialFunctionCall!(
   Export Float 4
   TextureObject=TextureObject() {samplerType=SAMPLERTYPE_Masks, texture=/Game/Athena/Environments/Nexus/Textures/NexFloor_M}
   TextureSize=@n9
   WorldPosition=@n11.PS
   World Space Normal=@n10.PS
   ProjectionTransitionContrast
   Pan_YZ=@n12
   Pan_XZ=@n14
   Pan_XY=@n15
  ) {materialFunction=&/WorldAlignedTexture_ComplexPan}
  n25 = ComponentMask(None=@n24.XYZ Texture) {g=true}
  n26 = MaterialFunctionCall(
   Transition Phase=Power(
    Base=MaterialFunctionCall(TextureObject=@n16, TextureSize=@n17, WorldPosition=@n11.PS, World Space Normal=@n10.PS, Pan_YZ=@n20, Pan_XZ=@n22, Pan_XY=@n23) {materialFunction=&WorldAlignedTexture_ComplexPan}.XYZ Texture
    Exp=ScalarParameter() {defaultValue=0.75, parameterName=Contrast_heighttransition}
   )
   Height Texture=OneMinus(None=@n25)
   Contrast=ScalarParameter() {defaultValue=0.1, parameterName=Contrast_heighttransition2}
  ) {materialFunction=&HeightLerp}
  n27 = MaterialFunctionCall(
   BaseNormal=GetMaterialAttributes(None=@n5) {attributeGetTypes [#1, #2, #3]}.Normal
   AdditionalNormal=LinearInterpolate(
    A=Constant3Vector() {constant=<rgba 0 0 1 0>}
    B=MaterialFunctionCall!(
     TextureObject=@n6
     TextureSize=@n9
     Normal=@n10.PS
     WorldSpace
     ProjectionTransitionContrast
     WorldPosition=@n11.PS
     Use High Quality Normals
     Pan_YZ=@n12
     Pan_XZ=@n14
     Pan_XY=@n15
    ) {materialFunction=&/WorldAlignedNormal_ComplexPan}.XYZ Texture
    Alpha=@n26.Alpha
   )
  ) {materialFunction=&BlendAngleCorrectedNormals}
  n28 = Transform(None=@n27.Result)
  n29 = Fresnel(ExponentIn=ScalarParameter() {defaultValue=3, parameterName=FresnelBaseColorExponent}, Normal=@n28) {baseReflectFraction=0}
  n30 = LinearInterpolate(
   A=Constant() {r=1}
   B=Multiply(A=ComponentMask(None=@n24.XYZ Texture) {r=true}, B=ScalarParameter() {defaultValue=0.75, parameterName=DarkenZeroPontMesh})
   Alpha=@n26.Alpha
  )
  n31 = ScalarParameter() {defaultValue=-1, parameterName=SpecularZeroPointMesh}
  n32 = LinearInterpolate(A=ScalarParameter() {defaultValue=0.5, parameterName=ChromaticBrightness}, B=Constant(), Alpha=@n26.Alpha)
  n33 = Fresnel(ExponentIn=ScalarParameter() {defaultValue=1.0666, parameterName=CromaticExponent}, Normal=@n28) {baseReflectFraction=0}
  n34 = Multiply(A=TextureSample(UVs=OneMinus(None=@n33)) {texture=&T_Inky_Smoke_Tile_Chromatic}.RGB, B=@n33)
  n35 = Constant3Vector()
  n36 = VectorParameter() {parameterName=EmissiveZeroPoint, defaultValue=<rgba 0.5833 1.3372 2 1>}
  n37 = Constant() {r=1}
  n38 = Power(
   Base=OneMinus(None=LinearInterpolate(A=@n37, B=@n25, Alpha=@n26.Alpha))
   Exp=ScalarParameter() {defaultValue=2, parameterName=ZeroPointHeightContrast}
  )
  n39 = MakeMaterialAttributes(
   BaseColor=Multiply(A=LinearInterpolate(A=@n3.RGB, B=@n4.RGB, Alpha=@n29), B=@n30)
   Specular=LinearInterpolate(A=Constant() {r=1}, B=Multiply(A=@n30, B=@n31), Alpha=@n26.Alpha)
   Roughness=ScalarParameter() {defaultValue=0.1, parameterName=R_ZeroPoint}
   EmissiveColor=Add(A=Multiply(A=@n32, B=@n34), B=LinearInterpolate(A=@n35, B=@n36.RGB, Alpha=@n38))
   Normal=@n27.Result
  )
  n40 = FunctionInput() {inputName="Customization Mask (_CM B)", inputType=FunctionInput_Scalar, sortPriority=24, previewValue=<1 0 0 1>}
  FunctionOutput(None=StaticSwitchParameter(True=BlendMaterialAttributes(A=@n5, B=@n39, Alpha=@n40), False=@n5) {group=Customization_Wrap, parameterName=Zero Point Wrap})
  ObjectRadius()
  ObjectRadius()
 ]}
