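# Material function for the Zero Point 2 wrap, which paints a glowing energy core over weapons, vehicles and other wrapped items.
# Camera-facing virtual-plane UVs centered on the object, offset per item type or blended to alternate UVs, drive every layer.
# A noise-distorted radial glow, counter-rotating twirls, Fresnel rim and optional panning stars form the emissive.
# A customization mask blends the result over incoming material attributes, behind a static switch.
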MaterialFunction /Game/Weapons/WeaponSkins/Wraps/MaterialFunctions/MF_ZeroPointWrap2
graph {nodes [
  n1 = ScalarParameter() {defaultValue=1, group=ZeroPoint_ExteriorNoise, parameterName=ZP_ExteriorStars1_Tiling}
  n2 = VectorParameter() {group=ZeroPoint_UV, parameterName=ZP_Offset_Vehicle}
  n3 = VectorParameter() {group=ZeroPoint_UV, parameterName=ZP_Offset_VehicleSide}
  n4 = Constant()
  n5 = Constant()
  n6 = FunctionInput(Preview=Constant()) {bUsePreviewValueAsDefault=true, inputName=isMech, inputType=FunctionInput_Scalar}
  n7 = MaterialFunctionCall(Color=MaterialFunctionCall(
    R=FunctionInput(Preview=@n4) {inputName=isVehicle, inputType=FunctionInput_Scalar, sortPriority=5}
    G=FunctionInput(Preview=@n5) {bUsePreviewValueAsDefault=true, inputName=Side Project Vehicle, inputType=FunctionInput_Scalar}
    B=@n6
   ) {materialFunction=&AppendMany}.RGB) {materialFunction=&SplitComponents}
  n8 = LinearInterpolate(
   A=LinearInterpolate(
    A=VectorParameter() {group=ZeroPoint_UV, parameterName=ZP_Offset_Weapon}.RGB
    B=LinearInterpolate(A=@n2.RGB, B=@n3.RGB, Alpha=@n7.G)
    Alpha=@n7.R
   )
   B=VectorParameter() {group=ZeroPoint_UV, parameterName=ZP_Offset_Mech}.RGB
   Alpha=@n7.B
  )
  n9 = PreSkinnedLocalBounds()
  n10 = Add(A=@n9.Min, B=FunctionInput() {inputName=Object Bounds Min Adjustment, sortPriority=51}) {constB=0}
  n11 = FunctionInput() {inputName=Object Bounds Max Adjustment, sortPriority=52}
  n12 = Multiply(A=Subtract(A=Add(A=@n9.Max, B=@n11) {constB=0}, B=@n10)) {constB=0.5}
  n13 = LinearInterpolate(A=TransformPosition(None=Add(A=@n10, B=@n12)) {transformType=TRANSFORMPOSSOURCE_World}, B=ObjectPositionWS(), Alpha=@n6)
  n14 = Normalize(VectorInput=Subtract(A=@n13, B=CameraPositionWS()))
  n15 = Normalize(VectorInput=CrossProduct(A=@n14, B=VectorParameter() {parameterName=Axis, defaultValue=<rgba 0 0 1 1>}.RGB))
  n16 = MaterialFunctionCall(Float3=PreSkinnedLocalBounds().Half Extents) {materialFunction=&BreakOutFloat3Components}
  n17 = Multiply(
   A=MaterialFunctionCall() {materialFunction=/Engine/Functions/Engine_MaterialFunctions02/WorldPositionOffset/ObjectScale}.Scale X
   B=Max(A=Max(A=@n16.R, B=@n16.R), B=@n16.B)
  )
  n18 = LinearInterpolate(
   A=Add(
    A=ComponentMask(None=@n8) {g=true, r=true}
    B=MaterialFunctionCall!(Plane Normal Axis=@n14, Plane X Axis=@n15, Plane Center=@n13, UV Scale=@n17) {materialFunction=/Game/Athena/Environments/Nexus/Materials/VirtualPlaneCoordinatesAbs}.UVs
   )
   B=NamedRerouteUsage()
   Alpha=Add(A=NamedRerouteUsage(), B=NamedRerouteUsage()) {constB=0}
  )
  n19 = TextureObjectParameter() {
   group=ZeroPoint_Textures
   parameterName=ZP2_DistortionMask
   samplerType=SAMPLERTYPE_Masks
   texture=&T_BlackMask
  }
  n20 = TextureSample(
   UVs=Panner(
    Coordinate=Multiply(A=@n18, B=@n1)
    Speed=ScalarParameter() {defaultValue=-0.1, group=ZeroPoint_ExteriorNoise, parameterName=ZP_ExteriorStars1_panSpeed}
   )
   Tex=@n19
  )
  n21 = ScalarParameter() {defaultValue=1, group=ZeroPoint_ExteriorNoise, parameterName=ZP_ExteriorStars2_ Tiling}
  n22 = TextureSample(
   UVs=Panner(
    Coordinate=Multiply(A=@n18, B=@n21)
    Speed=ScalarParameter() {defaultValue=0.2, group=ZeroPoint_ExteriorNoise, parameterName=ZP_ExteriorStars2_panSpeed}
   )
   Tex=@n19
  )
  n23 = Panner(Coordinate=TextureCoordinate() {uTiling=2, vTiling=2}) {speedX=0.015, speedY=-0.04}
  n24 = Multiply(
   A=Add(
    A=ChannelMaskParameter(None=@n20.RGBA) {group=ZeroPoint_ExteriorNoise, parameterName=ZP_ExteriorStars1_Channel, sortPriority=6}.R
    B=ChannelMaskParameter(None=@n22.RGBA) {group=ZeroPoint_ExteriorNoise, parameterName=ZP_ExteriorStars2_Channel, sortPriority=6}.R
   )
   B=ChannelMaskParameter(None=TextureSample(UVs=@n23, Tex=@n19).RGBA) {group=ZeroPoint_ExteriorNoise, parameterName=ZP_StarNoise_ChannelMask, sortPriority=99}.R
  )
  n25 = VectorParameter() {group=ZeroPoint_ExteriorNoise, parameterName=ZP_StarColor, defaultValue=<rgba 0 9.3228 20 1>}
  n26 = MaterialFunctionCall(
   UVs=@n18
   Texture Scale=LinearInterpolate(
    A=ScalarParameter() {defaultValue=1.9475, parameterName=ScaleTwirl}
    B=ScalarParameter() {defaultValue=1.9475, parameterName=ScaleTwirl_Vehicle}
    Alpha=@n7.R
   )
  ) {materialFunction=&ScaleUVsByCenter}
  n27 = Constant2Vector() {g=0.5, r=0.5}
  n28 = ScalarParameter() {defaultValue=-0.425, parameterName=ColorCoreTwirl_RotationSpeed}
  n29 = Time()
  n30 = TextureObjectParameter() {group=ZeroPoint_Textures, parameterName=ZP2_Mask, texture=&T_BlacksRGB}
  n31 = TextureSample(
   UVs=MaterialFunctionCall(UVs=@n26.UVs, Rotation Center=@n27, "Rotation Angle (0-1)"=Multiply(A=@n28, B=@n29)) {materialFunction=&CustomRotator}.Rotated Values
   Tex=@n30
  )
  n32 = ChannelMaskParameter(None=@n31.RGBA) {parameterName=ColorCoreTwirl_Channel, sortPriority=6, defaultValue=<rgba 0 0 1 0>}
  n33 = ChannelMaskParameter(None=@n31.RGBA) {maskChannel=Blue, parameterName=ColorCoreTwirlb_Channel, sortPriority=6, defaultValue=<rgba 0 0 1 0>}
  n34 = Add(A=@n32.R, B=@n33.R)
  n35 = ScalarParameter() {defaultValue=0.3, parameterName=ColorCoreTwirl2_RotationSpeed}
  n36 = TextureSample(
   UVs=MaterialFunctionCall(UVs=@n26.UVs, Rotation Center=@n27, "Rotation Angle (0-1)"=Multiply(A=@n29, B=@n35)) {materialFunction=&CustomRotator}.Rotated Values
   Tex=@n30
  )
  n37 = ChannelMaskParameter(None=@n36.RGBA) {maskChannel=Green, parameterName=ColorCoreTwirl2_Channel, sortPriority=6, defaultValue=<rgba 0 1 0 0>}
  n38 = ChannelMaskParameter(None=@n36.RGBA) {maskChannel=Blue, parameterName=ColorCoreTwirl2b_Channel, sortPriority=6, defaultValue=<rgba 0 0 1 0>}
  n39 = Add(A=@n37.R, B=@n38.R)
  n40 = VectorParameter() {parameterName=Fresnel_Color, defaultValue=<rgba 1 1 1 1>}
  n41 = FunctionInput() {inputName=Material, inputType=FunctionInput_MaterialAttributes, sortPriority=-200}
  n42 = Fresnel(
   ExponentIn=ScalarParameter() {defaultValue=8, parameterName=Fresnel -----}
   Normal=Transform(None=GetMaterialAttributes(None=@n41) {attributeGetTypes [#1, #2, #3]}.Normal)
  ) {baseReflectFraction=0}
  n43 = DotProduct(A=@n14, B=CameraVectorWS())
  n44 = OneMinus(None=Ceil(None=Saturate(None=@n43)))
  n45 = VectorParameter() {parameterName=EmColor1, defaultValue=<rgba 0.1329 0.2581 1 1>}
  n46 = VectorParameter() {parameterName=EmColor2, defaultValue=<rgba 0 0 0 1>}
  n47 = ScalarParameter() {defaultValue=0.3, parameterName=DistanceFalloff}
  n48 = Multiply(A=@n18, B=ScalarParameter() {defaultValue=1, parameterName=Distortion Tiling1})
  n49 = ScalarParameter() {defaultValue=-0.1, parameterName=panSpeed1}
  n50 = ChannelMaskParameter(None=TextureSample(UVs=Panner(Coordinate=@n48, Speed=@n49), Tex=@n19).RGBA) {parameterName=Distortion_Channel, sortPriority=6}
  n51 = Panner(
   Coordinate=Multiply(A=@n18, B=ScalarParameter() {defaultValue=1, parameterName=Distortion Tiling2})
   Speed=ScalarParameter() {defaultValue=0.2, parameterName=panSpeed2}
  )
  n52 = Add(A=@n50.R, B=ChannelMaskParameter(None=TextureSample(UVs=@n51, Tex=@n19).RGBA) {parameterName=Distortion2_Channel, sortPriority=6}.R)
  n53 = ShadingPathSwitch(Default=@n52, Deferred=@n52, Forward=@n50.R, Mobile=@n50.R)
  n54 = QualitySwitch(Default=@n53, High=@n53, Medium=@n53)
  n55 = ScalarParameter() {defaultValue=-0.1238, parameterName=Distortion}
  n56 = Constant2Vector() {g=0.5, r=0.5}
  n57 = Fresnel(ExponentIn=ScalarParameter() {defaultValue=0.4967, parameterName=DistortionExp}) {baseReflectFraction=0}
  n58 = ScalarParameter() {defaultValue=0.2744, parameterName=DistortionNormal}
  n59 = Multiply(A=Add(A=Multiply(A=@n54, B=@n55), B=Distance(A=@n18, B=@n56)), B=OneMinus(None=Multiply(A=@n57, B=@n58)))
  n60 = Clamp(
   None=Power(Base=Divide(A=@n59, B=@n47), Exp=ScalarParameter() {defaultValue=80, parameterName=FresnelEXP})
   Min=ScalarParameter() {defaultValue=0.9, parameterName=max}
  )
  n61 = Multiply(
   A=VectorParameter() {parameterName=CoreGLow, defaultValue=<rgba 0.4329 0.0848 5 1>}.RGB
   B=ScalarParameter() {defaultValue=0.4667, parameterName=Glow -----}
  )
  n62 = Divide(A=@n59, B=ScalarParameter() {defaultValue=1, parameterName=DistanceFalloff2})
  n63 = Clamp(None=Power(Base=OneMinus(None=@n62), Exp=ScalarParameter() {defaultValue=18, parameterName=FresnelEXP3}))
  n64 = ScalarParameter() {defaultValue=1.3, parameterName=DistanceFalloff3}
  n65 = Power(Base=OneMinus(None=Divide(A=@n59, B=@n64)), Exp=ScalarParameter() {defaultValue=3, parameterName=FresnelEXP4})
  n66 = Add(
   A=Add(A=LinearInterpolate(A=@n45.RGB, B=@n46.RGB, Alpha=@n60), B=Multiply(A=@n61, B=@n63))
   B=Multiply(A=VectorParameter() {parameterName=CoreGLow2, defaultValue=<rgba 0.1172 0.3089 0.4688 1>}.RGB, B=Clamp(None=@n65))
  )
  n67 = Add(
   A=Multiply(A=@n32.R, B=VectorParameter() {parameterName=ColorCoreTwirl, defaultValue=<rgba 0 0.6753 2 1>}.RGB)
   B=Multiply(A=VectorParameter() {parameterName=ColorCoreTwirlb, defaultValue=<rgba 0 0.9915 1 1>}.RGB, B=@n33.R)
  )
  n68 = Add(
   A=Multiply(A=@n37.R, B=VectorParameter() {parameterName=ColorCoreTwirl2, defaultValue=<rgba 0.8938 1 0 1>}.RGB)
   B=Multiply(A=@n38.R, B=VectorParameter() {parameterName=ColorCoreTwirl2b, defaultValue=<rgba 1 0 0.2603 1>}.RGB)
  )
  n69 = Desaturation(
   None=Power(Base=Add(A=@n67, B=@n68), Exp=ScalarParameter() {defaultValue=7.02, parameterName=ContrastZeroTwirl})
   Fraction=ScalarParameter() {defaultValue=-1.1005, parameterName=BrightnesSaturation}
  )
  n70 = ScalarParameter() {defaultValue=0.3333, parameterName=Brightnesswirl}
  n71 = Add(
   A=Multiply(A=Multiply(A=@n40.RGB, B=@n42), B=ScalarParameter() {defaultValue=1, parameterName=Fresnel Intensity})
   B=Add(A=Multiply(A=@n44, B=@n66), B=Multiply(A=@n69, B=@n70))
  ) {constB=0}
  n72 = Add(A=LinearInterpolate(A=Multiply(A=@n24, B=@n25.RGB), Alpha=Add(A=@n34, B=@n39)) {constB=0}, B=@n71)
  n73 = MakeMaterialAttributes!(
   BaseColor=VectorParameter() {parameterName=BaseColor}.RGB
   Metallic
   Specular=ScalarParameter() {defaultValue=0.5, parameterName=ZP2_Specular}
   Roughness=ScalarParameter() {defaultValue=0.1, parameterName=ZP2_Roughness}
   Anisotropy
   EmissiveColor=StaticSwitchParameter(True=@n72, False=@n71) {group=ZeroPoint_ExteriorNoise, parameterName=Use_ExteriorNoise, sortPriority=-100}
   Opacity
   OpacityMask
   Normal
   Tangent
   WorldPositionOffset
   SubsurfaceColor
   ClearCoat
   ClearCoatRoughness
   AmbientOcclusion
   Refraction
   CustomizedUVs_0
   CustomizedUVs_1
   CustomizedUVs_2
   CustomizedUVs_3
   CustomizedUVs_4
   CustomizedUVs_5
   CustomizedUVs_6
   CustomizedUVs_7
   PixelDepthOffset
   ShadingModel
   Displacement
  )
  n74 = FunctionInput() {inputName="Customization Mask (_CM B)", inputType=FunctionInput_Scalar, sortPriority=24, previewValue=<1 0 0 1>}
  n75 = FunctionInput() {inputName=Customization Mask B, inputType=FunctionInput_Scalar, sortPriority=24, previewValue=<1 0 0 1>}
  n76 = NamedRerouteUsage()
  n77 = NamedRerouteUsage()
  n78 = LinearInterpolate(A=@n74, B=@n75, Alpha=Saturate(None=Subtract(A=@n76, B=@n77)))
  FunctionOutput(None=StaticSwitchParameter(True=BlendMaterialAttributes(A=@n41, B=@n73, Alpha=@n78), False=@n41) {group=Customization_Wrap, parameterName=Zero Point Wrap2})
  LinearInterpolate(A=@n71, B=@n72, Alpha=ScalarParameter() {group=ZeroPoint_ExteriorNoise, parameterName=Use_ExteriorCoreNoise, sortPriority=-100})
  n79 = VectorParameter() {parameterName=Axis, defaultValue=<rgba 0 0 1 1>}
  MaterialFunctionCall(
   Plane Normal Axis=@n14
   Plane X Axis=Normalize(VectorInput=CrossProduct(A=@n14, B=@n79.RGB))
   Plane Center=@n13
   UV Scale=Multiply(A=@n17, B=ScalarParameter() {defaultValue=0.2, parameterName=SizeProjectionTexture2Core})
  ) {materialFunction=&VirtualPlaneCoordinatesAbs}
  n80 = TransformPosition(None=Constant3Vector()) {transformType=TRANSFORMPOSSOURCE_World}
  n81 = Normalize(VectorInput=Subtract(A=@n80, B=CameraPositionWS()))
  n82 = Constant3Vector() {constant=<rgba 0 0 1 1>}
  n83 = FunctionInput() {bUsePreviewValueAsDefault=true, inputName=Object-SafeZone-Virtual-Pivot, inputType=FunctionInput_Vector4, sortPriority=105}
  n84 = VectorParameter() {bUseCustomPrimitiveData=true, parameterName=Pliant_Crouch_Offset, primitiveDataIndex=7}
  n85 = Constant()
  n86 = LinearInterpolate(A=@n83, B=Add(A=@n83, B=AppendVector(A=@n84.RGB, B=@n85)), Alpha=Subtract(A=NamedRerouteUsage(), B=NamedRerouteUsage()))
  n87 = MaterialFunctionCall() {materialFunction=&ObjectScale}
  n88 = MaterialFunctionCall(Float3=PreSkinnedLocalBounds().Extents) {materialFunction=&BreakOutFloat3Components}
  n89 = Max(A=Max(A=@n88.R, B=@n88.R), B=@n88.B)
  MaterialFunctionCall(
   Plane Normal Axis=@n81
   Plane X Axis=Normalize(VectorInput=CrossProduct(A=@n81, B=@n82))
   Plane Center=Add(A=@n80, B=ComponentMask(None=@n86) {b=true, g=true, r=true})
   UV Scale=Multiply(A=ComponentMask(None=@n86) {a=true}, B=Multiply(A=@n87.Scale X, B=@n89))
  ) {materialFunction=&VirtualPlaneCoordinatesAbs}
  n90 = ScalarParameter() {defaultValue=-0.4201, parameterName=ZeroMask1}
  n91 = ScalarParameter() {defaultValue=2, parameterName=ZeroMask3}
  n92 = FunctionInput() {inputName="Customization Crevasse (_CM G)", inputType=FunctionInput_Scalar, sortPriority=21, previewValue=<1 0 0 1>}
  Multiply(A=Saturate(None=LinearInterpolate(A=@n90, B=@n91, Alpha=@n92)), B=@n78)
  NamedRerouteDeclaration(None=FunctionInput() {bUsePreviewValueAsDefault=true, inputName=isCharacter, inputType=FunctionInput_Scalar}) {name=IsCharacter, nodeColor=<rgba 0.7944 1 0.0641 1>}
  n93 = FunctionInput() {bUsePreviewValueAsDefault=true, inputName=IsStatsB, inputType=FunctionInput_Vector4, sortPriority=-19}
  NamedRerouteDeclaration(None=ComponentMask(None=@n93) {r=true}) {name=IsBackpack, nodeColor=<rgba 0 0.9253 1 1>}
  NamedRerouteDeclaration(None=ComponentMask(None=@n93) {g=true}) {name=IsPickaxe, nodeColor=<rgba 0 0.9253 1 1>}
  NamedRerouteDeclaration(None=ComponentMask(None=@n93) {a=true}) {name=isCharAccent, nodeColor=<rgba 0 0.9253 1 1>}
  n94 = FunctionInput() {bUsePreviewValueAsDefault=true, inputName=Object-SafeZone-Virtual-CH-Switch, inputType=FunctionInput_Scalar, sortPriority=105}
  NamedRerouteDeclaration(None=MaterialFunctionCall!(
    Camera _pos
    Axis
    Scale
    Scale_Multplyer
    Object-SafeZone-Virtual-Pivot=@n83
    Object-SafeZone-Virtual-CH-Switch=@n94
    SinglePlane Normal Axis
    SinglePlane X Axis
    PlaneCenter_distortion
   ) {materialFunction=&/MF_SafeZone_VirtualPlane}.Uvs) {name=SafeZoneVP_UVs, nodeColor=<rgba 1 0 0.6562 1>}
  ObjectRadius()
  n95 = Distance(A=@n13, B=WorldPosition().XYZ)
  n96 = Distance(A=ComponentMask(None=CameraPositionWS()) {g=true, r=true}, B=ComponentMask(None=@n13) {g=true, r=true})
  OneMinus(None=Clamp(None=Divide(A=@n95, B=@n96)))
  n97 = Multiply(A=ScalarParameter() {defaultValue=10, parameterName=RotationSpeedCore}, B=Time())
  Rotator(Coordinate=@n26.UVs, Time=@n97) {speed=-0.25}
  Rotator(Coordinate=@n26.UVs, Time=@n97) {speed=0.2133}
  ScalarParameter() {defaultValue=1185.5463, parameterName=SizeProjectionTextureCore}
 ]}
