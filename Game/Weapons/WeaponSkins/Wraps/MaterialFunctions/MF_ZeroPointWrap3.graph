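# Zero Point cosmetic wrap material function that blends a metallic kaleidoscope surface with an emissive glow over the input material through the customization mask, behind a static switch.
# Camera-facing virtual-plane UVs, with pivot adjustments and a SafeZone alternative, drive two counter-rotating mirror normals chosen by radial sections.
# The resulting normal's reflection vector samples chromatic smoke, which is masked by radial falloff and added to halo, core and Fresnel glows.
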MaterialFunction /Game/Weapons/WeaponSkins/Wraps/MaterialFunctions/MF_ZeroPointWrap3
graph {nodes [
  n1 = VectorParameter() {parameterName=ZeroPointFresnel, defaultValue=<rgba 0.2765 0.2782 0.2917 1>}
  n2 = FunctionInput() {inputName=Material, inputType=FunctionInput_MaterialAttributes, sortPriority=-200}
  n3 = Fresnel(
   ExponentIn=ScalarParameter() {defaultValue=4, parameterName=Fresnel}
   Normal=Transform(None=GetMaterialAttributes(None=@n2) {attributeGetTypes [#1, #2, #3]}.Normal)
  ) {baseReflectFraction=0}
  n4 = VectorParameter() {parameterName=ChromaticTint, defaultValue=<rgba 1.2031 1.4915 3 1>}
  n5 = ObjectPositionWS()
  n6 = Add(
   A=VectorParameter() {bUseCustomPrimitiveData=true, parameterName=Pliant_Crouch_Offset, primitiveDataIndex=7}.RGB
   B=VectorParameter() {parameterName=Pliant Object Position Adjust Character}.RGB
  )
  n7 = VectorParameter() {parameterName=Pliant Object Position Adjust Backpack}
  n8 = LinearInterpolate(
   A=@n5
   B=LinearInterpolate(A=Add(A=@n5, B=@n6) {constB=0}, B=Add(A=@n5, B=@n7.RGB) {constB=0}, Alpha=NamedRerouteUsage())
   Alpha=NamedRerouteUsage()
  )
  n9 = CameraPositionWS()
  n10 = Normalize(VectorInput=Subtract(A=@n8, B=@n9))
  n11 = CrossProduct(A=@n10, B=VectorParameter() {parameterName=Axis, defaultValue=<rgba 0 0 1 1>}.RGB)
  n12 = MaterialFunctionCall(Float3=PreSkinnedLocalBounds().Half Extents) {materialFunction=&BreakOutFloat3Components}
  n13 = Multiply(
   A=MaterialFunctionCall() {materialFunction=/Engine/Functions/Engine_MaterialFunctions02/WorldPositionOffset/ObjectScale}.Scale X
   B=Max(A=Max(A=@n12.R, B=@n12.R), B=@n12.B)
  )
  n14 = LinearInterpolate(
   A=ScalarParameter() {defaultValue=1, parameterName=UVRadialScaleZero}
   B=ScalarParameter() {defaultValue=1, parameterName=UVRadialScaleZero_Backpack}
   Alpha=NamedRerouteUsage()
  )
  n15 = PreSkinnedLocalBounds()
  n16 = Add(A=@n15.Min, B=FunctionInput() {inputName=Object Bounds Min Adjustment, sortPriority=51}) {constB=0}
  n17 = FunctionInput() {inputName=Object Bounds Max Adjustment, sortPriority=52}
  n18 = Multiply(A=Subtract(A=Add(A=@n15.Max, B=@n17) {constB=0}, B=@n16)) {constB=0.5}
  n19 = LinearInterpolate(
   A=TransformPosition(None=Add(A=@n16, B=@n18)) {transformType=TRANSFORMPOSSOURCE_World}
   B=@n8
   Alpha=FunctionInput() {bUsePreviewValueAsDefault=true, inputName=isMech, inputType=FunctionInput_Scalar}
  )
  n20 = LinearInterpolate(
   A=MaterialFunctionCall!(Plane Normal Axis=@n10, Plane X Axis=Normalize(VectorInput=@n11), Plane Center=@n19, UV Scale=Divide(A=@n13, B=@n14)) {materialFunction=/Game/Athena/Environments/Nexus/Materials/VirtualPlaneCoordinatesAbs}.UVs
   B=NamedRerouteUsage()
   Alpha=Add(A=NamedRerouteUsage(), B=NamedRerouteUsage())
  )
  n21 = Multiply(A=Time(), B=ScalarParameter() {defaultValue=0.25, parameterName=SpeedRotation})
  n22 = Rotator(Coordinate=@n20, Time=@n21)
  n23 = Rotator(Coordinate=@n20, Time=@n21) {speed=-0.25}
  n24 = Constant2Vector() {g=-0.5, r=-0.5}
  n25 = ComponentMask(None=MaterialFunctionCall(Vector or UVs=Add(A=@n20, B=@n24)) {materialFunction=&VectorToRadialValue}.Radial Coordinates) {r=true}
  n26 = ScalarParameter() {defaultValue=4, parameterName=Sections}
  n27 = Ceil(None=Sine(None=Multiply(A=@n25, B=@n26)))
  n28 = Transform(None=LinearInterpolate(
    A=TextureSample(UVs=@n22) {samplerType=SAMPLERTYPE_Normal, texture=/Game/Athena/Environments/Nexus/Textures/KaleidosMirror_N}.RGB
    B=TextureSample(UVs=@n23) {samplerType=SAMPLERTYPE_Normal, texture=&KaleidosMirror_N}.RGB
    Alpha=Saturate(None=@n27)
   )) {transformSourceType=TRANSFORMSOURCE_World, transformType=TRANSFORM_Tangent}
  n29 = ReflectionVectorWS(CustomWorldNormal=@n28)
  n30 = Distance(A=@n20, B=Constant2Vector() {g=0.5, r=0.5})
  n31 = ScalarParameter() {defaultValue=3.6483, parameterName=RadiusZeroPointGlow}
  n32 = OneMinus(None=Saturate(None=Divide(A=@n30, B=@n31)))
  n33 = Power(Base=@n32, Exp=ScalarParameter() {defaultValue=16, parameterName=FalloffZeroPointGlow})
  n34 = Multiply(
   A=TextureSample(UVs=ComponentMask(None=@n29) {g=true, r=true}) {texture=&T_Inky_Smoke_Tile_Chromatic}.RGB
   B=@n33
  )
  n35 = Multiply(A=@n33, B=VectorParameter() {parameterName=ZeroPointGlow, defaultValue=<rgba 0.3788 0.1564 0.5781 1>}.RGB)
  n36 = Multiply(
   A=Power(Base=@n32, Exp=ScalarParameter() {defaultValue=80, parameterName=FalloffZeroPointGlow2})
   B=VectorParameter() {parameterName=ZeroPointGlowCore, defaultValue=<rgba 0.8333 1.875 5 1>}.RGB
  )
  n37 = Power(Base=Add(A=Multiply(A=@n4.RGB, B=@n34), B=Add(A=@n35, B=@n36)), Exp=ScalarParameter() {defaultValue=1.6445, parameterName=ContrastColor_ZP})
  n38 = ScalarParameter() {defaultValue=2, parameterName=BrightnessColor_ZP}
  n39 = MakeMaterialAttributes!(
   BaseColor=VectorParameter() {parameterName=BaseColoKaleidos, defaultValue=<rgba 0.0785 0.0876 0.1667 1>}.RGB
   Metallic=ScalarParameter() {defaultValue=1, parameterName=MetallicKaleidos}
   Specular=Constant() {r=1}
   Roughness=Constant() {r=0.1}
   Anisotropy
   EmissiveColor=Add(A=Multiply(A=@n1.RGB, B=@n3), B=Multiply(A=@n37, B=@n38))
   Opacity
   OpacityMask
   Normal=@n28
   Tangent
   WorldPositionOffset
   SubsurfaceColor
   ClearCoat
   ClearCoatRoughness
   AmbientOcclusion
   Refraction
   CustomizedUVs_0
   CustomizedUVs_1
   CustomizedUVs_2
   CustomizedUVs_3
   CustomizedUVs_4
   CustomizedUVs_5
   CustomizedUVs_6
   CustomizedUVs_7
   PixelDepthOffset
   ShadingModel
   Displacement
  )
  n40 = FunctionInput() {inputName="Customization Mask (_CM B)", inputType=FunctionInput_Scalar, sortPriority=24, previewValue=<1 0 0 1>}
  FunctionOutput(None=StaticSwitchParameter(True=BlendMaterialAttributes(A=@n2, B=@n39, Alpha=@n40), False=@n2) {group=Customization_Wrap, parameterName=Zero Point Wrap3})
  n41 = Distance(A=@n19, B=@n9)
  LinearInterpolate(A=Multiply(A=CameraVectorWS()) {constB=-1}, B=@n10, Alpha=Saturate(None=Divide(A=@n41, B=@n13)))
  n42 = TransformPosition(None=Constant3Vector()) {transformType=TRANSFORMPOSSOURCE_World}
  n43 = Normalize(VectorInput=Subtract(A=@n42, B=CameraPositionWS()))
  n44 = Constant3Vector() {constant=<rgba 0 0 1 1>}
  n45 = FunctionInput() {bUsePreviewValueAsDefault=true, inputName=Object-SafeZone-Virtual-Pivot, inputType=FunctionInput_Vector4, sortPriority=105}
  n46 = MaterialFunctionCall() {materialFunction=&ObjectScale}
  n47 = MaterialFunctionCall(Float3=PreSkinnedLocalBounds().Extents) {materialFunction=&BreakOutFloat3Components}
  n48 = Max(A=Max(A=@n47.R, B=@n47.R), B=@n47.B)
  MaterialFunctionCall(
   Plane Normal Axis=@n43
   Plane X Axis=Normalize(VectorInput=CrossProduct(A=@n43, B=@n44))
   Plane Center=Add(A=@n42, B=ComponentMask(None=@n45) {b=true, g=true, r=true})
   UV Scale=Multiply(A=ComponentMask(None=@n45) {a=true}, B=Multiply(A=@n46.Scale X, B=@n48))
  ) {materialFunction=&VirtualPlaneCoordinatesAbs}
  n49 = ScalarParameter() {defaultValue=-0.4201, parameterName=ZeroMask1}
  n50 = ScalarParameter() {defaultValue=2, parameterName=ZeroMask3}
  n51 = FunctionInput() {inputName="Customization Crevasse (_CM G)", inputType=FunctionInput_Scalar, sortPriority=21, previewValue=<1 0 0 1>}
  Multiply(A=Saturate(None=LinearInterpolate(A=@n49, B=@n50, Alpha=@n51)), B=@n40)
  Multiply(A=@n13) {constB=2}
  NamedRerouteDeclaration(None=FunctionInput() {bUsePreviewValueAsDefault=true, inputName=IsCharacter, inputType=FunctionInput_Scalar}) {name=IsCharacter, nodeColor=<rgba 0 0.9253 1 1>}
  n52 = FunctionInput() {bUsePreviewValueAsDefault=true, inputName=IsStatsB, inputType=FunctionInput_Vector2}
  NamedRerouteDeclaration(None=ComponentMask(None=@n52) {r=true}) {name=IsBackpack, nodeColor=<rgba 0 0.9253 1 1>}
  NamedRerouteDeclaration(None=ComponentMask(None=@n52) {g=true}) {name=IsPickaxe, nodeColor=<rgba 0 0.9253 1 1>}
  n53 = FunctionInput() {bUsePreviewValueAsDefault=true, inputName=Object-SafeZone-Virtual-CH-Switch, inputType=FunctionInput_Scalar, sortPriority=105}
  NamedRerouteDeclaration(None=MaterialFunctionCall!(
    Camera _pos
    Axis
    Scale
    Scale_Multplyer
    Object-SafeZone-Virtual-Pivot=@n45
    Object-SafeZone-Virtual-CH-Switch=@n53
    SinglePlane Normal Axis
    SinglePlane X Axis
    PlaneCenter_distortion
   ) {materialFunction=&/MF_SafeZone_VirtualPlane}.Uvs) {name=SafeZoneVP_UVs, nodeColor=<rgba 1 0 0.6562 1>}
  NamedRerouteUsage()
  ObjectPositionWS()
  ObjectRadius()
  ObjectRadius()
 ]}
