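# Primary customization material function that applies wrap cosmetic effects to weapons, vehicles, mechs and characters in Fortress shared materials.
# It breaks the customization mask into primary, crevice, secondary and gmap channels and computes wrap bounds from pre-skinned position and object bound adjustments.
# Triplanar and top/side projections feed a long serial chain of per-wrap and effect functions, each receiving masks, bounds and object-type flags.
# It outputs the final material plus triplanar and top-side masks; some wrap branches hang off the main chain.
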MaterialFunction /Game/Packages/Fortress_SharedMaterials/Base_Material_Functions/BaseCustomizationFunctions/MF_Base_Customization_Primary
graph {nodes [
  FunctionInput() {bUsePreviewValueAsDefault=true, inputName=--------------------, inputType=FunctionInput_Scalar, sortPriority=40}
  FunctionInput() {bUsePreviewValueAsDefault=true, inputName=------------------, inputType=FunctionInput_Scalar, sortPriority=-19}
  n1 = FunctionInput() {bUsePreviewValueAsDefault=true, inputName=Object SBounds Min Adjustment}
  n2 = FunctionInput() {bUsePreviewValueAsDefault=true, inputName=Object Bounds Max Adjustment}
  n3 = FunctionInput() {bUsePreviewValueAsDefault=true, inputName=Object Bounds Center Adjustment}
  n4 = FunctionInput() {bUsePreviewValueAsDefault=true, inputName=Forward Axis}
  n5 = NamedRerouteDeclaration(None=FunctionInput() {bUsePreviewValueAsDefault=true, inputName=Pre-Skinned Local Position, sortPriority=41}) {name=PreSkin_Position, nodeColor=<rgba 0.6039 0.1176 0.4118 1>}
  n6 = NamedRerouteUsage()
  n7 = NamedRerouteDeclaration(None=FunctionInput() {bUsePreviewValueAsDefault=true, inputName=Pre-Skinned Local Normal, sortPriority=42}) {name=PreSkin_Normal, nodeColor=<rgba 0.3451 0.3098 1 1>}
  n8 = FunctionInput() {bUsePreviewValueAsDefault=true, inputName="Customization Mask (_CM RGBA)", inputType=FunctionInput_Vector4, sortPriority=26}
  n9 = MaterialFunctionCall(Float4=@n8) {materialFunction=&BreakOutFloat4Components}
  n10 = NamedRerouteDeclaration(None=@n9.B) {name=Mask_CM, nodeColor=<rgba 0 1 0.0859 1>}
  n11 = NamedRerouteDeclaration(None=@n9.A) {name=Mask_Gmap, nodeColor=<rgba 0.0094 0.0094 0.901 1>}
  n12 = NamedRerouteUsage()
  n13 = NamedRerouteUsage()
  n14 = NamedRerouteUsage()
  n15 = NamedRerouteUsage()
  n16 = NamedRerouteUsage()
  n17 = MaterialFunctionCall!(
   is Vehicle=@n16
   Object Bounds Max Adjustment=@n2
   Object Bounds Min Adjustment=@n1
   Object Bounds Center Adjustment=@n3
   isMech=NamedRerouteUsage()
   Side Project Vehicle=@n12
   Local Position=@n5
  ) {materialFunction=/Game/Weapons/WeaponSkins/Wraps/MaterialFunctions/MF_WrapBounds}
  n18 = NamedRerouteDeclaration(None=FunctionInput() {bUsePreviewValueAsDefault=true, inputName=Pre-Skinned Bounds, sortPriority=43}) {name=PreSkin_Bounds, nodeColor=<rgba 1 0.4219 0 1>}
  n19 = NamedRerouteDeclaration(None=FunctionInput() {inputName=Material, inputType=FunctionInput_MaterialAttributes, sortPriority=-200}) {name=Root_Material_Master, nodeColor=<rgba 1 0.6328 0 1>}
  n20 = FunctionInput() {bUsePreviewValueAsDefault=true, inputName=Wrap ScaleMultiplier, inputType=FunctionInput_Scalar, previewValue=<1 0 0 1>}
  n21 = FunctionInput() {bUsePreviewValueAsDefault=true, inputName=Rotation X Axis, inputType=FunctionInput_Scalar}
  n22 = Add(
   A=FunctionInput() {bUsePreviewValueAsDefault=true, inputName=Rotation Z Axis, inputType=FunctionInput_Scalar}
   B=LinearInterpolate(
    A=ScalarParameter() {group=Primary Triplaner, parameterName=Weapon Z Axis Rot, sortPriority=-34}
    B=ScalarParameter() {group=Primary Triplaner, parameterName=Vehicle Z Axis Rot, sortPriority=-33}
    Alpha=@n16
   )
  )
  n23 = NamedRerouteUsage()
  n24 = NamedRerouteDeclaration(None=@n9.R) {name=Mask_Primary, nodeColor=<rgba 0.0094 0.0094 0.901 1>}
  n25 = MaterialFunctionCall!(
   Material=@n19
   ------------------
   Rotation X Axis=@n21
   ForwardAxis=@n4
   isVehicle=@n16
   Wrap ScaleMultiplier=@n20
   Rotation Z Axis=@n22
   isMech=@n6
   -------------------
   "Customization Mask (_CM B)"=@n24
   Pre-Skinned Local Position=@n5
   Pre-Skinned Local Normal=@n7
   Per Obj Bounds=@n17.Result
   Pre-Skinned Bounds=@n18
   IsStatsB=NamedRerouteUsage()
   isCharacter=@n23
  ) {materialFunction=/Game/Weapons/WeaponSkins/Wraps/MaterialFunctions/MF_Customization_Wrap}
  n26 = MaterialFunctionCall!(
   Triplaners-Masks textures=@n25.Triplaner Mask
   Material=@n25.Triplaner Mask
   Root_Material=NamedRerouteUsage()
   IsState=NamedRerouteUsage()
   IsState_B=NamedRerouteUsage()
   Pre-Skinned Local Normal=NamedRerouteUsage()
   Calcated_Bounds=NamedRerouteUsage()
   Local Position=NamedRerouteUsage()
   Mask Secondary=NamedRerouteUsage()
   Mask Crevise=NamedRerouteUsage()
   TS Mask_01
   TS Mask_02
   Mask Primary=NamedRerouteUsage()
   Object-SafeZone-Back
   Object-SafeZone-Middle
   Liquid Mask Input
   Object-SafeZone-Front
  ) {materialFunction=/Game/Weapons/WeaponSkins/Wraps/MaterialFunctions/MF_TopSideProjection_Wrap_V02}
  n27 = MaterialFunctionCall!(
   Material=MaterialFunctionCall!(Material=@n26.Result, TopSide_Mask_01=@n26.Result, TopSide_Mask_02=@n26.Result, "Customization Mask (_CM B)"=@n24) {materialFunction=/Game/Weapons/WeaponSkins/Wraps/MaterialFunctions/MF_TopSide_Effect_Mask_TOD}.Result
   IsState=NamedRerouteUsage()
   IsState_B=NamedRerouteUsage()
   Pre-Skinned Local Normal=NamedRerouteUsage()
   Calcated_Bounds=NamedRerouteUsage()
   Mask Crevise
   Mask Secondary=NamedRerouteUsage()
   Mask Primary=NamedRerouteUsage()
  ) {materialFunction=/Game/Weapons/WeaponSkins/Wraps/MaterialFunctions/MF_Projection_SplitWeapon}
  n28 = NamedRerouteDeclaration(None=@n17.Result) {name=Caculated_Bounds, nodeColor=<rgba 0.465 1 0.0095 1>}
  n29 = MaterialFunctionCall!(
   Material=@n27.Result
   IsStateB=NamedRerouteUsage()
   IsState=NamedRerouteUsage()
   Pre-Skinned Bounds=@n28
   Pre-Skinned Local Normal=@n7
   TS_Mask_02=@n26.TS_Layer02_Mask
   TS_Mask_01=@n26.TS_Layer02_Mask
   "Customization Mask (_CM B)"=@n24
  ) {materialFunction=/Game/Weapons/WeaponSkins/Wraps/MaterialFunctions/MF_LayeredShapes}
  n30 = NamedRerouteDeclaration(None=@n9.G) {name=Mask_Crevice, nodeColor=<rgba 0 0.7865 0 1>}
  n31 = MaterialFunctionCall!(
   Material=MaterialFunctionCall!(
    Material=MaterialFunctionCall!(
     Material=@n29.Result
     ------------------
     PreSkinLocalPosition=@n5
     PreSkinnedLocalNormal=@n7
     "Customization Crevasse (_CM G)"=@n30
     "Customization Mask (_CM B)"=@n24
    ) {materialFunction=/Game/Weapons/WeaponSkins/Wraps/MaterialFunctions/MF_Customization_Crystalline}.Result
    isVehicle=@n16
    "Customization Mask (_CM B)"=@n24
    Pre-Skinned Local Position=@n5
    Pre-Skinned Local Normal=@n7
   ) {materialFunction=&/MF_WrapDecal}.Result
   IsVehicle=@n16
   IsStatsB=NamedRerouteUsage()
   IsCharacter=NamedRerouteUsage()
   ForwardAxis=@n4
   Wrap ScaleMultiplier=@n20
   PreSkinLocalPosition=@n5
   PreSkinnedLocalNormal=@n7
   "Customization Mask (_CM B)"=@n24
  ) {materialFunction=/Game/Weapons/WeaponSkins/Wraps/MaterialFunctions/MF_Customization_Storm}
  n32 = MaterialFunctionCall!(
   Object Bounds Min Adjustment=@n1
   Local Position=@n5
   Local Bounds=@n18
   isVehicle=@n16
   isMech=@n6
   Forward Axis=@n4
   Object Bounds Max Adjustment=@n2
   Material=@n31.Result
   Object Bounds Center Adjustment=@n3
   Root_Mat=NamedRerouteUsage()
   Side Project Vehicle=@n12
   "Customization Mask (_CM B)"=@n24
   "Customization Mask 2 (_CM B Top Right)"=@n10
   Pre-Skinned Local Normal=@n7
   IsStatsB=NamedRerouteUsage()
   isCharacter=NamedRerouteUsage()
  ) {materialFunction=/Game/Weapons/WeaponSkins/Wraps/MaterialFunctions/MF_CUstomization_Sideprojection.MF_Customization_Sideprojection}
  n33 = Saturate(None=Add(A=@n9.R, B=@n9.R))
  n34 = MaterialFunctionCall!(
   Primary Mask=@n26.Mask_Override
   Bounds=@n18
   Materials=MaterialFunctionCall!(
    GMap=@n11
    isMech=@n6
    isVehicle=@n16
    Curvature=@n30
    Material=MaterialFunctionCall!(
     GMap=@n11
     TopLayerMaterial_Custom
     BottomLayerMaterial_Custom
     Crevasse=@n30
     MaskPrimary=@n24
     isMech=@n6
     isVehicle=@n16
     Material=@n32.Result
     MaskSecondary=@n10
     MaskCombined=@n33
     PreskinnedPosition=@n5
     PreskinnedNormal=@n7
     isCharacter=@n13
     Root-Material=@n14
     IsStatsB=@n15
    ) {materialFunction=/Game/Weapons/WeaponSkins/Wraps/MaterialFunctions/MF_DirtAndEdgeWear}.Result
    Mask=@n24
    PreskinnedPosition=@n5
    PreskinnedBounds=@n18
    PreskinnedNormal=@n7
   ) {materialFunction=/Game/Weapons/WeaponSkins/Wraps/MaterialFunctions/MF_GMapGlowAndDistort}.Result
   isVehicle=@n16
   Local Position=@n5
   CUstomization Mask=@n24
   isMech=@n6
   Side Project Vehicle=@n12
   Pre-Skin_Bounds=@n28
  ) {materialFunction=/Game/Weapons/WeaponSkins/Wraps/MaterialFunctions/MF_SideGradientPanningEmissive}
  n35 = MaterialFunctionCall!(
   Material=MaterialFunctionCall!(
    Material=@n34.Result
    Material_Root=NamedRerouteUsage()
    isMech=@n6
    Side Project Vehicle=@n12
    Local Normals=@n7
    Local Position=@n5
    IsState_B=NamedRerouteUsage()
    IsState=NamedRerouteUsage()
    Calcated_Bounds=NamedRerouteUsage()
    Object Bounds Center Adjustment=@n3
    Object Bounds Min Adjustment=@n1
    Object Bounds Max Adjustment=@n2
    "Customization Mask (_CM B)"=@n24
   ) {materialFunction=/Game/Weapons/WeaponSkins/Wraps/MaterialFunctions/MF_ClockWork}.Result
   AO=@n11
   IsVehicle=@n16
   IsMech=@n6
   IsCharacter=NamedRerouteUsage()
   "Customization Crevasse (_CM G)"=@n30
   "Customization Mask (_CM B)"=@n24
  ) {materialFunction=/Game/Weapons/WeaponSkins/Wraps/MaterialFunctions/MF_OakWrap}
  n36 = MaterialFunctionCall!(
   Gmap=@n30
   PreskinnedBounds=@n18
   Isvehicle=@n16
   IsMech=@n6
   PreSkinnedPosition=@n5
   Material=@n35.Result
   Mask=@n33
   IsCharacter=NamedRerouteUsage()
   IsStatsB=NamedRerouteUsage()
  ) {materialFunction=/Game/Weapons/WeaponSkins/Wraps/MaterialFunctions/MF_SmokeyFX}
  n37 = MaterialFunctionCall!(Material=@n19) {materialFunction=/Engine/Functions/MaterialLayerFunctions/MatLayerBlend_BreakNormal}
  n38 = MaterialFunctionCall!(
   Material=MaterialFunctionCall!(
    Material=MaterialFunctionCall!(Materials=@n36.Result, Normal=@n37.Normal, Primary Mask=@n24) {materialFunction=/Game/Weapons/WeaponSkins/Wraps/MaterialFunctions/MF_FresnelEmissive}.Result
    Rotation Z Axis=@n22
    Rotation X Axis=@n21
    Tirplaner01Mask=@n26.Mask_Override
    Pre-Skinned Local Position=@n5
    Pre-Skinned Local Normal=@n7
    isVehicle=@n16
    isMech=@n6
    "Customization Mask (_CM B)"=@n33
   ) {materialFunction=/Game/Weapons/WeaponSkins/Wraps/MaterialFunctions/MF_BlendedPanningNormal}.Result
   Local Normals=@n7
   Forward Axis=@n4
   Rotation X Axis=@n21
   Rotation Z Axis=@n22
   Local Position=@n5
   "Customization Mask (_CM B Top left)"=@n24
   Gmap=@n11
   isVehicle=@n16
   isMech=@n6
   Object Bounds Center Adjustment=@n3
   Object Bounds Min Adjustment=@n1
   Object Bounds Max Adjustment=@n2
   IsStatsB=NamedRerouteUsage()
  ) {materialFunction=/Game/Weapons/WeaponSkins/Wraps/MaterialFunctions/MF_BuildupZVector}
  n39 = MaterialFunctionCall!(
   Material=@n38.Result
   ForwardAxis=@n4
   Side Project Vehicle=@n12
   isMech=@n6
   isVehicle=@n16
   Pre-Skinned Local Normal=@n7
   Local Position=@n5
   Object Bounds Min Adjustment=@n1
   Object Bounds Max Adjustment=@n2
   Object Bounds Center Adjustment=@n3
   "Customization Mask (_CM B)"=@n24
  ) {materialFunction=/Game/Weapons/WeaponSkins/Wraps/MaterialFunctions/MF_Fryangles}
  n40 = MaterialFunctionCall!(Material=@n39.Result, Bounds=@n28, isMech=@n6, isVehicle_Side=@n12, isVehicle=@n16, "Customization Mask (_CM B)"=@n24) {materialFunction=/Game/Weapons/WeaponSkins/Wraps/MaterialFunctions/MF_Wrap_NeonBands}
  n41 = MaterialFunctionCall!(
   Material=MaterialFunctionCall!(
    Material=@n40.Result
    Side Project Vehicle=@n12
    isMech=@n6
    isVehicle=@n16
    Pre-Skinned Local Normal=@n7
    Calcated_Bounds=@n28
    "Customization Mask (_CM B)"=@n24
   ) {materialFunction=/Game/Weapons/WeaponSkins/Wraps/MaterialFunctions/MF_TopSide_GradentToColorCurve}.Result
   Side Project Vehicle=@n12
   ForwardAxis=@n4
   isMech=@n6
   isVehicle=@n16
   Pre-Skinned Local Normal=@n7
   Calcated_Bounds=@n28
   Customization Mask Curvature=@n30
   "Customization Mask (_CM B)"=@n24
   Object Bounds Min Adjustment=@n1
   Object Bounds Max Adjustment=@n2
   IsStatsB=NamedRerouteUsage()
  ) {materialFunction=/Game/Weapons/WeaponSkins/Wraps/MaterialFunctions/MF_Wrap_FruitPunch}
  n42 = MaterialFunctionCall!(
   Material=@n41.Result_1
   IsStatsB=NamedRerouteUsage()
   IsState=NamedRerouteUsage()
   Pre-Skinned Local Normal=@n7
   Calcated_Bounds=@n28
   "Customization Mask (_CM B)"=@n24
   Curavture=@n30
   Object Bounds Max Adjustment=@n2
   Object Bounds Min Adjustment=@n1
   isCharAccent=NamedRerouteUsage()
   Object-SafeZone-Virtual-CH-Switch=NamedRerouteUsage()
   Object-SafeZone-Virtual-Pivot=NamedRerouteUsage()
  ) {materialFunction=/Game/Weapons/WeaponSkins/Wraps/MaterialFunctions/MF_Wrap_Amber}
  n43 = MaterialFunctionCall!(
   Material=MaterialFunctionCall!(
    Material=MaterialFunctionCall!(
     Material=@n42.Result
     ForwardAxis=@n4
     isMech=@n6
     isVehicle=@n16
     Side Project Vehicle=@n12
     Pre-Skinned Local Normal=@n7
     Local Position=@n5
     Wrap Bounds=@n28
     "Customization Mask (_CM B Top left)"=@n24
     "Customization Mask 2 (_CM B Top Right)"=@n10
    ) {materialFunction=/Game/Weapons/WeaponSkins/Wraps/MaterialFunctions/MF_Beats_Wrap}.Result
    IsVehicle=@n16
    IsMech=@n6
    Pre-Skinned Local Normal=@n7
    Calcated_Bounds=@n28
    "Customization Crevasse (_CM G)"=@n30
    "Customization Mask (_CM B)"=@n24
   ) {materialFunction=/Game/Weapons/WeaponSkins/Wraps/MaterialFunctions/MF_Wrap_HoloFoil}.Result
   Side Project Vehicle=@n12
   isCharacter=NamedRerouteUsage()
   isMech=@n6
   isVehicle=@n16
   Pre-Skinned Local Normal=@n7
   Calcated_Bounds=@n28
   Curavture=@n30
   Customization Mask B=@n24
   "Customization Mask (_CM B)"=@n33
   isCharAccent=NamedRerouteUsage()
  ) {materialFunction=/Game/Weapons/WeaponSkins/Wraps/MaterialFunctions/MF_Wrap_Frozen_Lake}
  n44 = MaterialFunctionCall!(
   Material=MaterialFunctionCall!(
    Material=@n43.Result
    isState=NamedRerouteUsage()
    Pre-Skinned Local Normal=@n7
    Calcated_Bounds=@n28
    "Customization Mask (_CM B)"=@n33
    Customization Mask B=@n24
    Curavture=@n30
    isCharAccent=NamedRerouteUsage()
   ) {materialFunction=/Game/Weapons/WeaponSkins/Wraps/MaterialFunctions/MF_Wrap_GasCloud}.Result
   isState=NamedRerouteUsage()
   Pre-Skinned Local Normal=@n7
   Calcated_Bounds=@n28
   Curavture=@n30
   Foward Axis=@n4
   "Customization Mask (_CM B)"=@n33
  ) {materialFunction=/Game/Weapons/WeaponSkins/Wraps/MaterialFunctions/MF_Wrap_Neon}
  n45 = MaterialFunctionCall!(
   Material=@n44.Result_1
   isState=NamedRerouteUsage()
   Pre-Skinned Local Normal=@n7
   Calcated_Bounds=@n28
   Curavture=@n30
   "Customization Mask (_CM A+B)"=@n33
   "Customization Mask (_CM B)"=@n24
   isCharAccent=NamedRerouteUsage()
  ) {materialFunction=/Game/Weapons/WeaponSkins/Wraps/MaterialFunctions/MF_Function_FlowMap}
  n46 = MaterialFunctionCall!(
   Material=MaterialFunctionCall!(
    Material=@n45.Result
    Material_Root=NamedRerouteUsage()
    IsState=NamedRerouteUsage()
    IsState_B=NamedRerouteUsage()
    Pre-Skinned Local Normal=@n7
    Calcated_Bounds=@n28
    Mask Crevise=NamedRerouteUsage()
    Mask Primary=NamedRerouteUsage()
    Mask Secondary=NamedRerouteUsage()
    Object Bounds Min Adjustment=@n1
    Object Bounds Max Adjustment=@n2
    Object-SafeZone-Virtual-CH-Switch=NamedRerouteUsage()
    Object-SafeZone-Virtual-Pivot=NamedRerouteUsage()
   ) {materialFunction=/Game/Weapons/WeaponSkins/Wraps/MaterialFunctions/MF_Wrap_FlapjackWrangler}.Result
   IsState=NamedRerouteUsage()
   Pre-Skinned Local Normal=@n7
   Calcated_Bounds=@n28
   "Customization Mask (_CM B Top left)"=@n24
   "Customization Mask 2 (_CM B Top Right)"=@n10
   isCharAccent=NamedRerouteUsage()
  ) {materialFunction=/Game/Weapons/WeaponSkins/Wraps/MaterialFunctions/MF_Wrap_SideMaskedEmissiveReactive}
  n47 = MaterialFunctionCall!(
   Material=MaterialFunctionCall!(
    Material=@n46.Result_1
    IsState=NamedRerouteUsage()
    IsStatsB=NamedRerouteUsage()
    IsCharacter=NamedRerouteUsage()
    Pre-Skinned Local Normal=@n7
    Pre-Skinned Local Position=@n5
    Calculated Bounds=NamedRerouteUsage()
    Gmap=@n11
    Curavture=@n30
    Object Bounds Max Adjustment=@n2
    Object Bounds Min Adjustment=@n1
    Object-SafeZone-Virtual-CH-Switch=NamedRerouteUsage()
    Object-SafeZone-Virtual-Pivot=NamedRerouteUsage()
   ) {materialFunction=/Game/Weapons/WeaponSkins/Wraps/MaterialFunctions/MF_Wrap_Liquid}.Result
   IsState=NamedRerouteUsage()
   IsState_B=NamedRerouteUsage()
   Pre-Skinned Local Normal=@n7
   Calcated_Bounds=@n28
   Curavture=@n30
   "Customization Mask (_CM B Top left)"=@n24
   "Customization Mask 2 (_CM B Top Right)"=@n10
  ) {materialFunction=/Game/Weapons/WeaponSkins/Wraps/MaterialFunctions/MF_Wrap_PinkToons}
  n48 = MaterialFunctionCall!(
   Material=MaterialFunctionCall!(
    Material=@n47.Result
    isCharacter=NamedRerouteUsage()
    Side Project Vehicle=@n12
    isMech=@n6
    isVehicle=@n16
    Pre-Skinned Local Normal=@n7
    Calcated_Bounds=@n28
    "Customization Mask (_CM B)"=@n33
    IsStatsB=NamedRerouteUsage()
   ) {materialFunction=/Game/Weapons/WeaponSkins/Wraps/MaterialFunctions/MF_Wrap_Flatline}.Result
   IsState=NamedRerouteUsage()
   IsVehicle=NamedRerouteUsage()
   IsMech=NamedRerouteUsage()
   Pre-Skinned Local Normal=@n7
   Calcated_Bounds=@n28
   Customization Mask B=@n24
   "Customization Mask (_CM B)"=@n33
   Curavture=@n30
   IsStatsB=NamedRerouteUsage()
  ) {materialFunction=/Game/Weapons/WeaponSkins/Wraps/MaterialFunctions/MF_Function_Parallax}
  n49 = MaterialFunctionCall!(
   Material=@n48.Result_1
   Side Project Vehicle=@n12
   Forward Axis=@n4
   isMech=@n6
   isVehicle=@n16
   Pre-Skinned Local Normal=@n7
   Calcated_Bounds=@n28
   "Customization Mask (_CM B)"=@n24
  ) {materialFunction=/Game/Weapons/WeaponSkins/Wraps/MaterialFunctions/MF_Function_NumberDisplay}
  n50 = MaterialFunctionCall!(
   Material=@n49.Result_1
   Side Project Vehicle=@n12
   ForwardAxis=@n4
   isMech=@n6
   isVehicle=@n16
   Pre-Skinned Local Normal=@n7
   Calcated_Bounds=@n28
   "Customization Mask (_CM B)"=@n10
   "Customization Mask (_CM B)B"=@n24
   Curavture=@n30
   isCharAccent=NamedRerouteUsage()
   isCharacter=NamedRerouteUsage()
  ) {materialFunction=/Game/Weapons/WeaponSkins/Wraps/MaterialFunctions/MF_Wrap_HeadC}
  n51 = MaterialFunctionCall!(
   Material=@n50.Result
   Side Project Vehicle=@n12
   isMech=@n6
   isVehicle=@n16
   Pre-Skinned Local Normal=@n7
   Calcated_Bounds=@n28
   "Customization Mask (_CM B)"=@n24
  ) {materialFunction=/Game/Weapons/WeaponSkins/Wraps/MaterialFunctions/MF_Wrap_HoloFoilTum.MF_Wrap_HoloFoilTUM}
  n52 = MaterialFunctionCall!(
   Material=@n51.Result
   IsStatsB=NamedRerouteUsage()
   Side Project Vehicle=NamedRerouteUsage()
   isMech=NamedRerouteUsage()
   isCharacter=NamedRerouteUsage()
   isVehicle=NamedRerouteUsage()
   "Customization Mask (_Secondary)"=@n10
   "Customization Mask (_CM B)"=@n24
   Object Bounds Min Adjustment=@n1
   Object Bounds Max Adjustment=@n2
   Object-SafeZone-Virtual-CH-Switch=NamedRerouteUsage()
   Object-SafeZone-Virtual-Pivot=NamedRerouteUsage()
  ) {materialFunction=/Game/Weapons/WeaponSkins/Wraps/MaterialFunctions/MF_Ruck}
  n53 = MaterialFunctionCall!(
   Material=@n52.Result
   IsState=NamedRerouteUsage()
   Side Project Vehicle=@n12
   isMech=@n6
   isVehicle=@n16
   Pre-Skinned Local Normal=@n7
   Calcated_Bounds=@n28
   "Customization Mask (_CM B)"=@n33
  ) {materialFunction=/Game/Weapons/WeaponSkins/Wraps/MaterialFunctions/MF_Wrap_BSequins}
  n54 = MaterialFunctionCall!(
   Material=@n53.Result
   IsStatsB=NamedRerouteUsage()
   ForwardAxis=@n4
   IsCharacter=NamedRerouteUsage()
   isMech=NamedRerouteUsage()
   isVehicle=NamedRerouteUsage()
   Side Project Vehicle=NamedRerouteUsage()
   Pre-Skinned Local Normal=NamedRerouteUsage()
   Local Position=@n5
   Object Bounds Min Adjustment=@n1
   Object Bounds Max Adjustment=@n2
   Object Bounds Center Adjustment=@n3
   "Customization Mask (_CM B)"=@n24
  ) {materialFunction=/Game/Weapons/WeaponSkins/Wraps/MaterialFunctions/MF_Pmint_Wrap}
  n55 = MaterialFunctionCall!(
   Material=@n54.Result
   Side Project Vehicle=NamedRerouteUsage()
   isMech=NamedRerouteUsage()
   isVehicle=NamedRerouteUsage()
   Pre-Skinned Local Normal=NamedRerouteUsage()
   Calcated_Bounds=NamedRerouteUsage()
   "Customization Mask (_CM B Top left)"=NamedRerouteUsage()
  ) {materialFunction=/Game/Weapons/WeaponSkins/Wraps/MaterialFunctions/MF_Wrap_EclipseS}
  n56 = MaterialFunctionCall!(
   Material_Root=NamedRerouteUsage()
   Material=@n55.Result
   IsState_B=NamedRerouteUsage()
   IsState=NamedRerouteUsage()
   Pre-Skinned Local Normal=NamedRerouteUsage()
   Calcated_Bounds=NamedRerouteUsage()
   Mask Crevise
   Mask Secondary=NamedRerouteUsage()
   Mask Primary=NamedRerouteUsage()
   Object-SafeZone-Front=NamedRerouteUsage()
  ) {materialFunction=/Game/Weapons/WeaponSkins/Wraps/MaterialFunctions/MF_Function_TheEye}
  n57 = MaterialFunctionCall!(
   Material=@n56.Result
   IsState=NamedRerouteUsage()
   IsState_B=NamedRerouteUsage()
   Pre-Skinned Local Normal=NamedRerouteUsage()
   Calcated_Bounds=NamedRerouteUsage()
   Mask Secondary=NamedRerouteUsage()
   Mask Crevise=NamedRerouteUsage()
   Mask Primary=NamedRerouteUsage()
  ) {materialFunction=/Game/Weapons/WeaponSkins/Wraps/MaterialFunctions/MF_Effect_GradentShape}
  n58 = MaterialFunctionCall!(
   Material=@n57.Result
   IsState_B=NamedRerouteUsage()
   IsState=NamedRerouteUsage()
   Pre-Skinned Local Normal=NamedRerouteUsage()
   Pre-Skinned Local Position=NamedRerouteUsage()
   Calcated_Bounds=NamedRerouteUsage()
   Mask Crevise=NamedRerouteUsage()
   Mask Secondary=NamedRerouteUsage()
   Mask Primary=NamedRerouteUsage()
   Object-SafeZone-VP-Character-PreView-Switch=NamedRerouteUsage()
   Object-SafeZone-Virtual-Pivot=NamedRerouteUsage()
  ) {materialFunction=/Game/Weapons/WeaponSkins/Wraps/MaterialFunctions/MF_Wrap_DarkAz}
  n59 = MaterialFunctionCall!(
   Material=@n58.Result
   IsState_B=NamedRerouteUsage()
   IsState=NamedRerouteUsage()
   Pre-Skinned Local Normal=NamedRerouteUsage()
   Calcated_Bounds=NamedRerouteUsage()
   Mask Crevise=NamedRerouteUsage()
   Mask Secondary=NamedRerouteUsage()
   Mask Primary=NamedRerouteUsage()
   Object-SafeZone-VP-Character-PreView-Switch=NamedRerouteUsage()
   Object-SafeZone-Virtual-Pivot=NamedRerouteUsage()
  ) {materialFunction=/Game/Weapons/WeaponSkins/Wraps/MaterialFunctions/MF_Wrap_Defect}
  n60 = MaterialFunctionCall!(
   TS_Emis_02_Msk=NamedRerouteUsage()
   Material=@n59.Result
   TS_Emis_01_Msk=NamedRerouteUsage()
   Material_Root=NamedRerouteUsage()
   IsState=NamedRerouteUsage()
   IsState_B=NamedRerouteUsage()
   Pre-Skinned Local Normal=NamedRerouteUsage()
   Calcated_Bounds=NamedRerouteUsage()
   Mask Secondary=NamedRerouteUsage()
   Mask Crevise=NamedRerouteUsage()
   Mask Primary=NamedRerouteUsage()
   Object-SafeZone-VP-Character-PreView-Switch=NamedRerouteUsage()
   Object-SafeZone-Virtual-Pivot=NamedRerouteUsage()
  ) {materialFunction=/Game/Weapons/WeaponSkins/Wraps/MaterialFunctions/MF_Wrap_NebulaV2}
  n61 = MaterialFunctionCall!(
   Material_1=@n60.Result
   Material _Root=NamedRerouteUsage()
   IsState_B=NamedRerouteUsage()
   IsState=NamedRerouteUsage()
   Pre-Skinned Local Normal=NamedRerouteUsage()
   Calcated_Bounds=NamedRerouteUsage()
   Mask Crevise=NamedRerouteUsage()
   Mask Secondary=NamedRerouteUsage()
   Mask Primary=NamedRerouteUsage()
   Object-SafeZone-VP-Character-PreView-Switch=NamedRerouteUsage()
   Object-SafeZone-Virtual-Pivot=NamedRerouteUsage()
  ) {materialFunction=/Game/Weapons/WeaponSkins/Wraps/MaterialFunctions/MF_Wrap_PlasmaCore}
  n62 = MaterialFunctionCall!(Material=@n61.Result, "Customization Mask (_CM B)"=@n24) {materialFunction=/Game/Weapons/WeaponSkins/Wraps/MaterialFunctions/MF_CritterManiac_Aged_FilmGrain_Wrap}
  n63 = NamedRerouteUsage()
  n64 = NamedRerouteUsage()
  n65 = NamedRerouteUsage()
  n66 = NamedRerouteUsage()
  n67 = NamedRerouteUsage()
  n68 = NamedRerouteUsage()
  n69 = NamedRerouteUsage()
  n70 = NamedRerouteUsage()
  n71 = NamedRerouteUsage()
  n72 = MaterialFunctionCall!(
   Material=MaterialFunctionCall!(
    Root_Matieral=NamedRerouteUsage()
    Material=MaterialFunctionCall!(
     Material=@n62.Result
     IsState_B=@n63
     IsState=@n64
     Pre-Skinned Local Normal=@n65
     Calcated_Bounds=@n66
     Mask Crevise=@n67
     Mask Secondary=@n68
     Mask Primary=@n69
     TS_Mask_01=@n70
     TS_Mask_02=@n71
    ) {materialFunction=/Game/Weapons/WeaponSkins/Wraps/MaterialFunctions/MF_Effect_Multicolor}.Result
    "Msk_Ts01<--->Ts02"
    IsState_B=NamedRerouteUsage()
    IsState=NamedRerouteUsage()
    CM_Msk=NamedRerouteUsage()
    Pre-Skinned Local Normal=NamedRerouteUsage()
    Calcated_Bounds=NamedRerouteUsage()
    Object-SafeZone-Back=NamedRerouteUsage()
    Object-SafeZone-Middle=NamedRerouteUsage()
    Object-SafeZone-Front=NamedRerouteUsage()
   ) {materialFunction=/Game/Weapons/WeaponSkins/Wraps/MaterialFunctions/MF_Effect_Decal_Reactive}.Result
   Decal_Mask
   Mask Crevise=NamedRerouteUsage()
   "Customization Mask (_CM B)"=@n24
   Mask Primary=NamedRerouteUsage()
   Mask Secondary=NamedRerouteUsage()
   TS_Mask_01=NamedRerouteUsage()
   TS_Mask_03
   TS_Mask_04
   TS_Mask_02=NamedRerouteUsage()
  ) {materialFunction=/Game/Weapons/WeaponSkins/Wraps/MaterialFunctions/MF_Effect_Iridescent}
  n73 = MaterialFunctionCall!(
   Material=MaterialFunctionCall!(
    In=MaterialFunctionCall!(In=@n72.Result, "Customization Mask (_CM B)"=@n24) {materialFunction=/Game/Weapons/WeaponSkins/Wraps/MaterialFunctions/MF_Effect_Reactive_Emissive}.Result
    IsState=NamedRerouteUsage()
    IsState_B=NamedRerouteUsage()
    Pre-Skinned Local Normal=NamedRerouteUsage()
    Calcated_Bounds=NamedRerouteUsage()
    Mask Crevise=NamedRerouteUsage()
    Mask Secondary=NamedRerouteUsage()
    Mask Primary=NamedRerouteUsage()
   ) {materialFunction=/Game/Weapons/WeaponSkins/Wraps/MaterialFunctions/MF_Effect_Visor}.Result
   IsState_B=NamedRerouteUsage()
   IsState=NamedRerouteUsage()
   Pre-Skinned Local Normal=NamedRerouteUsage()
   Calcated_Bounds=NamedRerouteUsage()
   Mask Crevise=NamedRerouteUsage()
   Mask Secondary=NamedRerouteUsage()
   Mask Gmap=NamedRerouteUsage()
   Mask Primary=NamedRerouteUsage()
   TS_Mask_02=NamedRerouteUsage()
   TS_Mask_01=NamedRerouteUsage()
  ) {materialFunction=/Game/Weapons/WeaponSkins/Wraps/MaterialFunctions/MF_EffectLiquidMetal}
  n74 = NamedRerouteUsage()
  n75 = NamedRerouteUsage()
  n76 = NamedRerouteUsage()
  n77 = NamedRerouteUsage()
  n78 = NamedRerouteUsage()
  n79 = MaterialFunctionCall!(
   Material=MaterialFunctionCall!(
    Material=MaterialFunctionCall!(Material=@n73.Result, Mask Crevise=@n74, Mask Primary=@n75, Mask Secondary=@n76, TS_Mask_02=@n77, TS_Mask_01=@n78, Left_Right_Mask) {materialFunction=/Game/Weapons/WeaponSkins/Wraps/MaterialFunctions/MF_Effect_Metal_LUT}.Result
    Customization Mask A=@n24
    Curavture=NamedRerouteUsage()
    Customization Mask B=@n10
   ) {materialFunction=/Game/Weapons/WeaponSkins/Wraps/MaterialFunctions/MF_Effect_Toon}.Result
   IsState=NamedRerouteUsage()
   IsState_B=NamedRerouteUsage()
   Pre-Skinned Local Normal=NamedRerouteUsage()
   Calcated_Bounds=NamedRerouteUsage()
   Mask Crevise=NamedRerouteUsage()
   Mask Secondary=NamedRerouteUsage()
   Mask Primary=NamedRerouteUsage()
  ) {materialFunction=/Game/Weapons/WeaponSkins/Wraps/MaterialFunctions/MF_Effect_AnimeShading}
  FunctionOutput(None=MaterialFunctionCall!(
    Material=MaterialFunctionCall!(Material=@n79.accent, "Customization Mask (_CM B)"=@n24) {materialFunction=/Game/Weapons/WeaponSkins/Wraps/MaterialFunctions/MF_Effect_Hue_Shift}.Result
    "Customization Mask (_CM B)"=@n24
   ) {materialFunction=/Game/Weapons/WeaponSkins/Wraps/MaterialFunctions/MF_Effect_Audio_Pulse}.Result)
  FunctionOutput(None=@n26.Mask_Override) {outputName=Triplaner Mask, sortPriority=4}
  FunctionOutput(None=@n25.Triplaner 01 Alpha) {outputName=Triplaner 01 Alpha, sortPriority=5}
  FunctionOutput(None=@n25.Triplaner 02 Alpha) {outputName=Triplaner 02 Alpha, sortPriority=6}
  FunctionOutput(None=NamedRerouteDeclaration(None=@n26.TS_Layer01_Mask) {name=TS_Mask_01, nodeColor=<rgba 1 0 0.7031 1>}) {outputName=TS_Mask_01}
  FunctionOutput(None=NamedRerouteDeclaration(None=@n26.TS_Layer02_Mask) {name=TS_Mask_02, nodeColor=<rgba 1 0 0.7031 1>}) {outputName=TS_Mask_02}
  FunctionOutput(None=NamedRerouteDeclaration(None=@n26.Emissive 01 Msk) {name=TS_Emis_01_Msk, nodeColor=<rgba 1 0.8672 0 1>}) {outputName=TS_Emis_01_Msk}
  FunctionOutput(None=NamedRerouteDeclaration(None=@n26.Emissive 02 Msk) {name=TS_Emis_02_Msk, nodeColor=<rgba 1 0.8672 0 1>}) {outputName=TS_Emis_02_Msk}
  MaterialFunctionCall!(
   Material=MaterialFunctionCall!(
    Material=MaterialFunctionCall!(
     Material=@n32.Result
     Local Position=@n5
     isVehicle=@n16
     isMech=@n6
     Object Bounds Max Adjustment=@n2
     Object Bounds Min Adjustment=@n1
     Object Bounds Center Adjustment=@n3
     Pre-Skinned Local Normal=@n7
     "Customization Mask (_CM B)"=@n24
    ) {materialFunction=/Game/Weapons/WeaponSkins/Wraps/MaterialFunctions/MF_Wrap_Equalizer}.Result
    Pre-Skinned Local Position=@n5
    RotXAxis=@n21
    isVehicle=@n16
    IsCharacter=NamedRerouteUsage()
    IsStatsB=NamedRerouteUsage()
    RotZAxis=@n22
    IsStats
    Calcated_Bounds
    "Customization Crevasse (_CM G)"=@n30
    "Customization Mask (_CM B)"=@n33
    Pre-Skinned Local Normal=@n7
    Pre-Skinned Bounds=@n18
    Object-SafeZone-VP-Character-PreView-Switch
    Object-SafeZone-Virtual-Pivot
   ) {materialFunction=/Game/Weapons/WeaponSkins/Wraps/MaterialFunctions/MF_Customization_TheCube}.Result
   IsCharacter=NamedRerouteUsage()
   IsStatsB=NamedRerouteUsage()
   IsVehicle=@n16
   Wrap ScaleMultiplier=@n20
   ForwardAxis=@n4
   PreSkinnedLocalNormal=@n7
   PreSkinLocalPosition=@n5
  ) {materialFunction=/Game/Weapons/WeaponSkins/Wraps/MaterialFunctions/MF_Customization_WeirdObjects}
  n80 = NamedRerouteUsage()
  MaterialFunctionCall!(
   Material=MaterialFunctionCall!(
    Material=MaterialFunctionCall!(
     Material=@n50.Result
     isState=@n80
     Pre-Skinned Local Normal=@n7
     Calcated_Bounds=@n28
     "Customization Mask (_CM B Top left)"=@n10
     "Customization Mask 2 (_CM B Top Right)"=@n24
    ) {materialFunction=/Game/Weapons/WeaponSkins/Wraps/MaterialFunctions/MF_Wrap_Alakazam}.Result
    isState=NamedRerouteUsage()
    Pre-Skinned Local Normal=@n7
    Calcated_Bounds=@n28
    Gmap=@n11
    "Customization Mask (_CM B Top left)"=@n10
    "Customization Mask 2 (_CM B Top Right)"=@n24
    IsStatsB=NamedRerouteUsage()
   ) {materialFunction=/Game/Weapons/WeaponSkins/Wraps/MaterialFunctions/MF_Wrap_Flora}.Result
   ForwardAxis=@n4
   isState=NamedRerouteUsage()
   Pre-Skinned Local Normal=@n7
   Calcated_Bounds=@n28
   Customization Mask A=@n24
   Curavture=@n30
   Customization Mask B=@n10
   Object Bounds Min Adjustment=@n1
   Object Bounds Max Adjustment=@n2
   Object-SafeZone-Virtual-CH-Switch=NamedRerouteUsage()
   Object-SafeZone-Virtual-Pivot=NamedRerouteUsage()
  ) {materialFunction=/Game/Weapons/WeaponSkins/Wraps/MaterialFunctions/MF_Wrap_BeeHive}
  n81 = MaterialFunctionCall!(
   Material=@n51.Result
   isSates=NamedRerouteUsage()
   Pre-Skinned Local Normal=@n7
   Calcated_Bounds=@n28
   Curavture=@n30
   Customization Mask B=@n24
   Customization Mask A=@n33
   isCharAccent=NamedRerouteUsage()
  ) {materialFunction=/Game/Weapons/WeaponSkins/Wraps/MaterialFunctions/MF_Wrap_AstroMap}
  n82 = MaterialFunctionCall!(
   Material=MaterialFunctionCall!(
    Material=MaterialFunctionCall!(
     Material=@n81.Result
     Side Project Vehicle=@n12
     isMech=@n6
     isVehicle=@n16
     Pre-Skinned Local Normal=@n7
     Calcated_Bounds=@n28
     "Customization Mask (_CM B Top left)"=@n10
     "Customization Mask 2 (_CM B Top Right)"=@n24
    ) {materialFunction=/Game/Weapons/WeaponSkins/Wraps/MaterialFunctions/MF_Wrap_FKeepers}.Result
    isState=NamedRerouteUsage()
    Pre-Skinned Local Normal=@n7
    Calcated_Bounds=@n28
    "Customization Mask (_CM B Top left)"=@n24
    "Customization Mask 2 (_CM B Top Right)"=@n10
    IsStatsB=NamedRerouteUsage()
   ) {materialFunction=/Game/Weapons/WeaponSkins/Wraps/MaterialFunctions/MF_Wrap_StrangeCosmos}.Result
   isState=NamedRerouteUsage()
   Pre-Skinned Local Normal=@n7
   Calcated_Bounds=@n28
   Curavture=@n30
   Customization Mask A=@n24
   Customization Mask B=@n10
  ) {materialFunction=/Game/Weapons/WeaponSkins/Wraps/MaterialFunctions/MF_Wrap_GoldenSkeleton}
  n83 = NamedRerouteUsage()
  n84 = MaterialFunctionCall!(
   Material=MaterialFunctionCall!(
    Material=MaterialFunctionCall!(Material=@n82.Result, IsState=@n83, Pre-Skinned Local Normal=@n7, Calcated_Bounds=@n28, Curavture=@n30, "Customization Mask (_CM B)"=@n24) {materialFunction=/Game/Weapons/WeaponSkins/Wraps/MaterialFunctions/MF_Function_LanternFest}.Result_1
    IsStatsB=NamedRerouteUsage()
    IsState=NamedRerouteUsage()
    Pre-Skinned Local Normal=@n7
    Calcated_Bounds=@n28
    Customization Mask B=@n10
    Customization Mask A=@n24
    Curavture=@n30
   ) {materialFunction=/Game/Weapons/WeaponSkins/Wraps/MaterialFunctions/MF_Wrap_LavaLamp}.Result
   isState=NamedRerouteUsage()
   Pre-Skinned Local Normal=@n7
   Calcated_Bounds=@n28
   "Customization Mask (_CM B)"=@n24
  ) {materialFunction=/Game/Weapons/WeaponSkins/Wraps/MaterialFunctions/MF_Wrap_SpaceCud}
  n85 = NamedRerouteUsage()
  n86 = NamedRerouteUsage()
  MaterialFunctionCall!(
   Material=MaterialFunctionCall!(
    Material=MaterialFunctionCall!(
     Material=@n84.Result
     IsState=@n85
     Pre-Skinned Local Normal=@n7
     Calcated_Bounds=@n28
     "Customization Mask (_CM B Top left)"=@n24
     "Customization Mask 2 (_CM B Top Right)"=@n10
     isCharAccent=@n86
    ) {materialFunction=/Game/Weapons/WeaponSkins/Wraps/MaterialFunctions/MF_Wrap_RetroRainbow}.Result
    Side Project Vehicle=@n12
    isMech=@n6
    isVehicle=@n16
    Pre-Skinned Local Normal=@n7
    Calcated_Bounds=@n28
    "Customization Mask (_CM B Top left)"=@n24
    "Customization Mask 2 (_CM B Top Right)"=@n10
   ) {materialFunction=/Game/Weapons/WeaponSkins/Wraps/MaterialFunctions/MF_Wrap_VaderIn}.Result
   IsState=NamedRerouteUsage()
   Pre-Skinned Local Normal=@n7
   Calcated_Bounds=@n28
   "Customization Mask (_CM B)B"=@n24
   "Customization Mask (_CM B)"=@n33
   isCharAccent=NamedRerouteUsage()
  ) {materialFunction=/Game/Weapons/WeaponSkins/Wraps/MaterialFunctions/MF_Wrap_Faux}
  MaterialFunctionCall!(
   Material=@n52.Result
   Side Project Vehicle=@n12
   isMech=@n6
   IsState=NamedRerouteUsage()
   isVehicle=@n16
   Pre-Skinned Local Normal=@n7
   Calcated_Bounds=@n28
   "Customization Mask (_CM B Top left)"=@n24
   "Customization Mask 2 (_CM B Top Right)"=@n10
  ) {materialFunction=/Game/Weapons/WeaponSkins/Wraps/MaterialFunctions/MF_Wrap_BeamU}
  n87 = FunctionInput() {bUsePreviewValueAsDefault=true, inputName=IsStates, inputType=FunctionInput_Vector4}
  n88 = MaterialFunctionCall(Float4=@n87) {materialFunction=&BreakOutFloat4Components}
  MaterialFunctionCall(
   X=NamedRerouteDeclaration(None=@n88.R) {name=isVehicle, nodeColor=<rgba 0 0.9253 1 1>}
   Y=NamedRerouteDeclaration(None=@n88.G) {name=isVehicleSide, nodeColor=<rgba 0 0.9253 1 1>}
   Z=NamedRerouteDeclaration(None=@n88.B) {name=isMech, nodeColor=<rgba 0 0.9253 1 1>}
   A=NamedRerouteDeclaration(None=@n88.A) {name=isCharacter, nodeColor=<rgba 0 0.9253 1 1>}
  ) {materialFunction=&MakeFloat4}
  n89 = FunctionInput() {bUsePreviewValueAsDefault=true, inputName=IsStatesB V4, inputType=FunctionInput_Vector4}
  n90 = MaterialFunctionCall(Float4=@n89) {materialFunction=&BreakOutFloat4Components}
  MaterialFunctionCall(
   X=NamedRerouteDeclaration(None=@n90.R) {name=IsBackpack, nodeColor=<rgba 0 0.9253 1 1>}
   Y=NamedRerouteDeclaration(None=@n90.G) {name=IsPickaxe, nodeColor=<rgba 0 0.9253 1 1>}
  ) {materialFunction=&MakeFloat2}
  n91 = MaterialFunctionCall!(
   Material=MaterialFunctionCall!(
    Material=@n53.Result
    IsStatsB=NamedRerouteUsage()
    Side Project Vehicle=NamedRerouteUsage()
    isMech=NamedRerouteUsage()
    isCharacter=NamedRerouteUsage()
    isVehicle=NamedRerouteUsage()
    Customization Mask B=@n33
    "Customization Mask (_CM B)"=@n24
    Object Bounds Min Adjustment=NamedRerouteUsage()
    Object Bounds Max Adjustment=NamedRerouteUsage()
    Object-SafeZone-Virtual-Pivot=NamedRerouteUsage()
    Object-SafeZone-Virtual-Pivot_CH_Switch=NamedRerouteUsage()
   ) {materialFunction=/Game/Weapons/WeaponSkins/Wraps/MaterialFunctions/MF_Wrap_CelestialGlow2}.Result
   IsStatsB=NamedRerouteUsage()
   ForwardAxis=@n4
   IsCharacter=NamedRerouteUsage()
   isMech=NamedRerouteUsage()
   isVehicle=NamedRerouteUsage()
   Side Project Vehicle=NamedRerouteUsage()
   Pre-Skinned Local Normal=@n7
   Local Position=@n5
   Object Bounds Min Adjustment=@n1
   Object Bounds Max Adjustment=@n2
   Object Bounds Center Adjustment=@n3
   "Customization Mask (_CM B)"=@n24
  ) {materialFunction=/Game/Weapons/WeaponSkins/Wraps/MaterialFunctions/MF_Cubes_Wrap}
  n92 = NamedRerouteUsage()
  n93 = NamedRerouteUsage()
  n94 = NamedRerouteUsage()
  MaterialFunctionCall!(
   Material=MaterialFunctionCall!(
    Material=MaterialFunctionCall!(Material=@n91.Result, Bounds=@n28, isMech=@n92, isVehicle_Side=@n93, isVehicle=@n94, "Customization Mask (_CM B)"=@n24) {materialFunction=/Game/Weapons/WeaponSkins/Wraps/MaterialFunctions/MF_Wrap_GhostHunter}.Result
    Local Normals=NamedRerouteUsage()
    Local Position=NamedRerouteUsage()
    isVehicle=NamedRerouteUsage()
    isMech=NamedRerouteUsage()
    isCharacter=NamedRerouteUsage()
    Side Project Vehicle=NamedRerouteUsage()
    Object Bounds Min Adjustment=@n1
    Object Bounds Max Adjustment=@n2
    Object Bounds Center Adjustment=@n3
    Customization Mask B=@n24
    "Customization Mask (_CM B)"=@n33
   ) {materialFunction=/Game/Weapons/WeaponSkins/Wraps/MaterialFunctions/MF_Wrap_AlDay}.Result
   IsState_B=NamedRerouteUsage()
   IsState=NamedRerouteUsage()
   Pre-Skinned Local Normal=NamedRerouteUsage()
   Calcated_Bounds=NamedRerouteUsage()
   Mask Secondary=NamedRerouteUsage()
   Mask Primary=NamedRerouteUsage()
  ) {materialFunction=/Game/Weapons/WeaponSkins/Wraps/MaterialFunctions/MF_Wrap_WLights}
  n95 = NamedRerouteUsage()
  n96 = NamedRerouteUsage()
  n97 = NamedRerouteUsage()
  n98 = NamedRerouteUsage()
  n99 = NamedRerouteUsage()
  n100 = NamedRerouteUsage()
  n101 = NamedRerouteUsage()
  n102 = NamedRerouteUsage()
  n103 = NamedRerouteUsage()
  n104 = NamedRerouteUsage()
  n105 = NamedRerouteUsage()
  MaterialFunctionCall!(
   Material=MaterialFunctionCall!(
    Material=MaterialFunctionCall!(
     Material=@n54.Result
     IsState=@n95
     IsState_B=@n96
     Pre-Skinned Local Normal=@n97
     Calcated_Bounds=@n98
     Mask Secondary=@n99
     Mask Gmap=@n100
     Mask Primary=@n101
     Object Bounds Min Adjustment=@n102
     Object Bounds Max Adjustment=@n103
     Object-SafeZone-VP_Charcter_Preview_Switch=@n104
     Object-SafeZone-Virtual-Pivot=@n105
    ) {materialFunction=/Game/Weapons/WeaponSkins/Wraps/MaterialFunctions/MF_Wrap_SkullPunk}.Result
    IsState=NamedRerouteUsage()
    IsState_B=NamedRerouteUsage()
    Pre-Skinned Local Normal=NamedRerouteUsage()
    Calcated_Bounds=NamedRerouteUsage()
    Mask Secondary=NamedRerouteUsage()
    Mask Crevise=NamedRerouteUsage()
    Mask Primary=NamedRerouteUsage()
   ) {materialFunction=/Game/Weapons/WeaponSkins/Wraps/MaterialFunctions/MF_Wrap_SlipStream}.Result
   IsState=NamedRerouteUsage()
   IsState_B=NamedRerouteUsage()
   Forward_Axis=NamedRerouteUsage()
   Pre-Skinned Local Normal=NamedRerouteUsage()
   Pre-Skinned Local Postion=NamedRerouteUsage()
   Calcated_Bounds=NamedRerouteUsage()
   Mask Crevise=NamedRerouteUsage()
   Mask Primary=NamedRerouteUsage()
   Mask Secondary=NamedRerouteUsage()
   Object-SafeZone-Virtual-Pivot=NamedRerouteUsage()
   Object-SafeZone-VP_Charcter_Preview_Switch=NamedRerouteUsage()
  ) {materialFunction=/Game/Weapons/WeaponSkins/Wraps/MaterialFunctions/MF_Wrap_HexColor}
  MaterialFunctionCall!(
   Material=@n55.Result
   IsState_B=NamedRerouteUsage()
   IsState=NamedRerouteUsage()
   Pre-Skinned Local Normal=NamedRerouteUsage()
   Calcated_Bounds=NamedRerouteUsage()
   Mask Crevise=NamedRerouteUsage()
   Mask Secondary=NamedRerouteUsage()
   Mask Primary=NamedRerouteUsage()
   Object-SafeZone-Front=NamedRerouteUsage()
  ) {materialFunction=/Game/Weapons/WeaponSkins/Wraps/MaterialFunctions/MF_Wrap_WDClouds}
  MaterialFunctionCall!(
   Material=MaterialFunctionCall!(
    Material=@n56.Result
    IsState=NamedRerouteUsage()
    IsState_B=NamedRerouteUsage()
    Pre-Skinned Local Normal=NamedRerouteUsage()
    Calcated_Bounds=NamedRerouteUsage()
    Mask Crevise=NamedRerouteUsage()
    Mask Secondary=NamedRerouteUsage()
    Mask Primary=NamedRerouteUsage()
   ) {materialFunction=/Game/Weapons/WeaponSkins/Wraps/MaterialFunctions/MF_Wrap_Disarray}.Result
   IsState=NamedRerouteUsage()
   IsState_B=NamedRerouteUsage()
   Pre-Skinned Local Normal=NamedRerouteUsage()
   Calcated_Bounds=NamedRerouteUsage()
   Mask Crevise=NamedRerouteUsage()
   Mask Secondary=NamedRerouteUsage()
   Mask Primary=NamedRerouteUsage()
  ) {materialFunction=/Game/Weapons/WeaponSkins/Wraps/MaterialFunctions/MF_Wrap_Cact_Rocker}
  n106 = MaterialFunctionCall!(
   Material=MaterialFunctionCall!(
    Material=@n57.Result
    IsState_B=NamedRerouteUsage()
    IsState=NamedRerouteUsage()
    Pre-Skinned Local Normal=NamedRerouteUsage()
    Calcated_Bounds=NamedRerouteUsage()
    Mask Crevise=NamedRerouteUsage()
    Mask Secondary=NamedRerouteUsage()
    Mask Primary=NamedRerouteUsage()
    Object-SafeZone-Front
   ) {materialFunction=/Game/Weapons/WeaponSkins/Wraps/MaterialFunctions/MF_Wrap_EqualizerBounce}.Result
   IsState=NamedRerouteUsage()
   IsState_B=NamedRerouteUsage()
   Pre-Skinned Local Normal=NamedRerouteUsage()
   Calcated_Bounds=NamedRerouteUsage()
   Mask Crevise=NamedRerouteUsage()
   Mask Secondary=NamedRerouteUsage()
   Mask Primary=NamedRerouteUsage()
  ) {materialFunction=/Game/Weapons/WeaponSkins/Wraps/MaterialFunctions/MF_Wrap_NeonGraffitiLava}
  n107 = NamedRerouteUsage()
  n108 = NamedRerouteUsage()
  n109 = NamedRerouteUsage()
  n110 = NamedRerouteUsage()
  n111 = NamedRerouteUsage()
  n112 = NamedRerouteUsage()
  n113 = NamedRerouteUsage()
  n114 = MaterialFunctionCall!(
   Material=MaterialFunctionCall!(
    Material=MaterialFunctionCall!(
     Material=@n106.Result
     IsState=@n107
     IsState_B=@n108
     Pre-Skinned Local Normal=@n109
     Calcated_Bounds=@n110
     Mask Crevise=@n111
     Mask Secondary=@n112
     Mask Primary=@n113
    ) {materialFunction=/Game/Weapons/WeaponSkins/Wraps/MaterialFunctions/MF_Wrap_GameBud}.Result
    IsState=NamedRerouteUsage()
    IsState_B=NamedRerouteUsage()
    Pre-Skinned Local Normal=NamedRerouteUsage()
    Calcated_Bounds=NamedRerouteUsage()
    Mask Crevise=NamedRerouteUsage()
    Mask Secondary=NamedRerouteUsage()
    Mask Primary=NamedRerouteUsage()
   ) {materialFunction=/Game/Weapons/WeaponSkins/Wraps/MaterialFunctions/MF_WrapLancelot}.Result
   IsState_B=NamedRerouteUsage()
   IsState=NamedRerouteUsage()
   Pre-Skinned Local Normal=NamedRerouteUsage()
   Calcated_Bounds=NamedRerouteUsage()
   Mask Crevise=NamedRerouteUsage()
   Mask Secondary=NamedRerouteUsage()
   Mask Gmap=NamedRerouteUsage()
   Mask Primary=NamedRerouteUsage()
  ) {materialFunction=/Game/Weapons/WeaponSkins/Wraps/MaterialFunctions/MF_WrapMitos}
  n115 = NamedRerouteUsage()
  n116 = NamedRerouteUsage()
  n117 = NamedRerouteUsage()
  n118 = NamedRerouteUsage()
  n119 = NamedRerouteUsage()
  n120 = NamedRerouteUsage()
  n121 = NamedRerouteUsage()
  n122 = MaterialFunctionCall!(
   Material=MaterialFunctionCall!(
    Material=MaterialFunctionCall!(
     Material=@n114.Result_1
     IsState=@n115
     IsState_B=@n116
     Pre-Skinned Local Normal=@n117
     Calcated_Bounds=@n118
     Mask Secondary=@n119
     Mask Crevise=@n120
     Mask Primary=@n121
    ) {materialFunction=/Game/Weapons/WeaponSkins/Wraps/MaterialFunctions/MF_Wrap_Wave}.Result
    IsState=NamedRerouteUsage()
    IsState_B=NamedRerouteUsage()
    Pre-Skinned Local Normal=NamedRerouteUsage()
    Calcated_Bounds=NamedRerouteUsage()
    Mask Crevise=NamedRerouteUsage()
    Mask Secondary=NamedRerouteUsage()
    Mask Primary=NamedRerouteUsage()
   ) {materialFunction=/Game/Weapons/WeaponSkins/Wraps/MaterialFunctions/MF_Barium}.Result
   IsState=NamedRerouteUsage()
   IsState_B=NamedRerouteUsage()
   Pre-Skinned Local Normal=NamedRerouteUsage()
   Calcated_Bounds=NamedRerouteUsage()
   Mask Secondary=NamedRerouteUsage()
   Mask Crevise=NamedRerouteUsage()
   Mask Primary=NamedRerouteUsage()
  ) {materialFunction=/Game/Weapons/WeaponSkins/Wraps/MaterialFunctions/MF_Wrap_CharlotteSummer}
  n123 = NamedRerouteUsage()
  n124 = NamedRerouteUsage()
  n125 = NamedRerouteUsage()
  n126 = NamedRerouteUsage()
  n127 = NamedRerouteUsage()
  n128 = NamedRerouteUsage()
  n129 = NamedRerouteUsage()
  n130 = NamedRerouteUsage()
  n131 = NamedRerouteUsage()
  MaterialFunctionCall!(
   Material=MaterialFunctionCall!(
    Material=MaterialFunctionCall!(
     Material=@n122.Result
     IsState_B=@n123
     IsState=@n124
     Pre-Skinned Local Normal=@n125
     Calcated_Bounds=@n126
     Mask Crevise=@n127
     Mask Secondary=@n128
     Mask Primary=@n129
     Object-SafeZone-VP-Character-PreView-Switch=@n130
     Object-SafeZone-Virtual-Pivot=@n131
    ) {materialFunction=/Game/Weapons/WeaponSkins/Wraps/MaterialFunctions/MF_Wrap_BadB}.Result
    IsState=NamedRerouteUsage()
    IsState_B=NamedRerouteUsage()
    Pre-Skinned Local Normal=NamedRerouteUsage()
    Calcated_Bounds=NamedRerouteUsage()
    Mask Crevise=NamedRerouteUsage()
    Mask Secondary=NamedRerouteUsage()
    Mask Primary=NamedRerouteUsage()
    ForwardAxis=NamedRerouteUsage()
   ) {materialFunction=/Game/Weapons/WeaponSkins/Wraps/MaterialFunctions/MF_Wrap_Candor}.Result
   IsState_B=NamedRerouteUsage()
   IsState=NamedRerouteUsage()
   Pre-Skinned Local Normal=NamedRerouteUsage()
   Calcated_Bounds=NamedRerouteUsage()
   Mask Crevise=NamedRerouteUsage()
   Mask Secondary=NamedRerouteUsage()
   Mask Primary=NamedRerouteUsage()
   ForwardAxis=NamedRerouteUsage()
  ) {materialFunction=/Game/Weapons/WeaponSkins/Wraps/MaterialFunctions/MF_Wrap_RenRaiderSpark}
  MaterialFunctionCall!(
   Material=MaterialFunctionCall!(
    Material=@n58.Result
    IsState=NamedRerouteUsage()
    IsState_B=NamedRerouteUsage()
    Pre-Skinned Local Normal=NamedRerouteUsage()
    Calcated_Bounds=NamedRerouteUsage()
    Mask Crevise=NamedRerouteUsage()
    Mask Secondary=NamedRerouteUsage()
    Mask Primary=NamedRerouteUsage()
   ) {materialFunction=/Game/Weapons/WeaponSkins/Wraps/MaterialFunctions/MF_Wrap_Cons}.Result
   IsState=NamedRerouteUsage()
   IsState_B=NamedRerouteUsage()
   Pre-Skinned Local Normal=NamedRerouteUsage()
   Calcated_Bounds=NamedRerouteUsage()
   Mask Crevise=NamedRerouteUsage()
   Mask Secondary=NamedRerouteUsage()
   Mask Primary=NamedRerouteUsage()
  ) {materialFunction=/Game/Weapons/WeaponSkins/Wraps/MaterialFunctions/MF_Wrap_VGlitch}
  n132 = MaterialFunctionCall!(
   Material=@n59.Result
   IsState=NamedRerouteUsage()
   IsState_B=NamedRerouteUsage()
   Pre-Skinned Local Normal=NamedRerouteUsage()
   Calcated_Bounds=NamedRerouteUsage()
   Mask Crevise=NamedRerouteUsage()
   Mask Secondary=NamedRerouteUsage()
   Mask Primary=NamedRerouteUsage()
  ) {materialFunction=/Game/Weapons/WeaponSkins/Wraps/MaterialFunctions/MF_Wrap_NeonBlob}
  n133 = NamedRerouteUsage()
  n134 = NamedRerouteUsage()
  n135 = NamedRerouteUsage()
  n136 = NamedRerouteUsage()
  n137 = NamedRerouteUsage()
  n138 = NamedRerouteUsage()
  n139 = NamedRerouteUsage()
  n140 = NamedRerouteUsage()
  n141 = NamedRerouteUsage()
  n142 = NamedRerouteUsage()
  n143 = NamedRerouteUsage()
  n144 = NamedRerouteUsage()
  n145 = NamedRerouteUsage()
  MaterialFunctionCall!(
   Material=MaterialFunctionCall!(
    Material=MaterialFunctionCall!(
     Material=@n132.Result
     IsState_B=@n133
     IsState=@n134
     Pre-Skinned Local Normal=@n135
     Pre-Skinned Position=@n136
     Calcated_Bounds=@n137
     Mask Crevise=@n138
     Mask Primary=@n139
     Mask Secondary=@n140
     Bounds_Max=@n141
     Bounds_min=@n142
     Object-SafeZone-Front=@n143
     Object-SafeZone-Middle=@n144
     Object-SafeZone-Back=@n145
    ) {materialFunction=/Game/Weapons/WeaponSkins/Wraps/MaterialFunctions/MF_Wrap_MagicMead}.Result
    Primary Mask=NamedRerouteUsage()
    Secondary Mask_1=NamedRerouteUsage()
    In CM Mask=NamedRerouteUsage()
    Mask GMAP=NamedRerouteUsage()
    Mask Crevice=NamedRerouteUsage()
   ) {materialFunction=/Game/Weapons/WeaponSkins/Wraps/MaterialFunctions/MF_Wrap_RadishMemory}.Result
   Material Root=NamedRerouteUsage()
   IsState_B=NamedRerouteUsage()
   IsState=NamedRerouteUsage()
   Pre-Skinned Local Normal=NamedRerouteUsage()
   Calcated_Bounds=NamedRerouteUsage()
   Mask Crevise=NamedRerouteUsage()
   Mask Secondary=NamedRerouteUsage()
   Mask Primary=NamedRerouteUsage()
  ) {materialFunction=/Game/Weapons/WeaponSkins/Wraps/MaterialFunctions/MF_ShinySt}
  n146 = NamedRerouteUsage()
  MaterialFunctionCall!(
   Material=MaterialFunctionCall!(
    Material=MaterialFunctionCall!(Material=@n34.Result, isVehicle=@n16, "Customization Mask (_CM B)"=@n33) {materialFunction=/Game/Weapons/WeaponSkins/Wraps/MaterialFunctions/MF_ZeroPointWrap}.Result
    IsStatsB=NamedRerouteUsage()
    Side Project Vehicle=@n12
    isCharacter=NamedRerouteUsage()
    isMech=@n6
    isVehicle=@n16
    "Customization Crevasse (_CM G)"=@n30
    Customization Mask B=@n24
    "Customization Mask (_CM B)"=@n33
    Object Bounds Min Adjustment=@n1
    Object Bounds Max Adjustment=@n2
    Object-SafeZone-Virtual-CH-Switch=@n146
    Object-SafeZone-Virtual-Pivot=NamedRerouteUsage()
   ) {materialFunction=/Game/Weapons/WeaponSkins/Wraps/MaterialFunctions/MF_ZeroPointWrap2}.Result
   isMech=@n6
   IsStatsB=NamedRerouteUsage()
   IsCharacter=NamedRerouteUsage()
   "Customization Crevasse (_CM G)"=@n30
   "Customization Mask (_CM B)"=@n33
   Object Bounds Min Adjustment=@n1
   Object Bounds Max Adjustment=@n2
   Object-SafeZone-Virtual-CH-Switch=@n146
   Object-SafeZone-Virtual-Pivot=NamedRerouteUsage()
  ) {materialFunction=/Game/Weapons/WeaponSkins/Wraps/MaterialFunctions/MF_ZeroPointWrap3}
  MaterialFunctionCall!(
   Material=MaterialFunctionCall!(
    Material=@n60.Result
    Material_Root=NamedRerouteUsage()
    IsState_B=NamedRerouteUsage()
    IsState=NamedRerouteUsage()
    Pre-Skinned Local Normal=NamedRerouteUsage()
    Calcated_Bounds=NamedRerouteUsage()
    Mask Crevise=NamedRerouteUsage()
    Mask Secondary=NamedRerouteUsage()
    Mask Primary=NamedRerouteUsage()
   ) {materialFunction=/Game/Weapons/WeaponSkins/Wraps/MaterialFunctions/MF_NitroF}.Result
   Material_Root=NamedRerouteUsage()
   IsState_B=NamedRerouteUsage()
   IsState=NamedRerouteUsage()
   Pre-Skinned Local Normal=NamedRerouteUsage()
   Calcated_Bounds=NamedRerouteUsage()
   Mask Crevise=NamedRerouteUsage()
   Mask Secondary=NamedRerouteUsage()
   Mask Primary=NamedRerouteUsage()
   Object-SafeZone-VP-Character-PreView-Switch=NamedRerouteUsage()
   Object-SafeZone-Virtual-Pivot=NamedRerouteUsage()
  ) {materialFunction=/Game/Weapons/WeaponSkins/Wraps/MaterialFunctions/MF_Wrap_GalaxyKn}
  MaterialFunctionCall!(
   Material_1=@n61.Result
   Material_Root=NamedRerouteUsage()
   IsState_B=NamedRerouteUsage()
   IsState=NamedRerouteUsage()
   Pre-Skinned Local Normal=NamedRerouteUsage()
   Calcated_Bounds=NamedRerouteUsage()
   Mask Crevise=NamedRerouteUsage()
   Mask Secondary=NamedRerouteUsage()
   Mask Primary=NamedRerouteUsage()
   Object-SafeZone-VP-Character-PreView-Switch=NamedRerouteUsage()
   Object-SafeZone-Virtual-Pivot=NamedRerouteUsage()
  ) {materialFunction=/Game/Weapons/WeaponSkins/Wraps/MaterialFunctions/MF_Wrap_TigerR}
  MaterialFunctionCall(Z=Constant(), A=NamedRerouteDeclaration(None=@n90.A) {name=isCharAccent, nodeColor=<rgba 0 0.9253 1 1>}) {materialFunction=&MakeFloat4}
  MaterialFunctionCall!(
   IsVehicle=@n16
   IsMech=@n6
   Pre-SkinnedLocalPosition=@n5
   Pre-SkinnedLocalBounds
   Pre-SkinnedLocalBounds_Min
   MaterialInput=MaterialFunctionCall!(
    PreskinnedPosition=@n5
    PreskinnedBounds=@n18
    isVehicle=@n16
    isMech=@n6
    Material=@n34.Result
    IsStatsB=NamedRerouteUsage()
    IsCharacter=NamedRerouteUsage()
    "Customization Mask (_CM B)"=@n33
    Object Bounds Min Adjustment=@n1
    Object Bounds Max Adjustment=@n2
   ) {materialFunction=/Game/Weapons/WeaponSkins/MaterialFunction/MF_HauntWrap}.Result
   IsStatsB=NamedRerouteUsage()
  ) {materialFunction=/Game/Weapons/WeaponSkins/Wraps/MaterialFunctions/MF_TechOpsBlue}
  MaterialFunctionCall!(
   Material=@n25.Result
   Triplaners-Masks textures=@n25.Result
   ForwardAxis=@n4
   Material_Bypass=@n19
   Pre-Skinned Local Normal=@n7
   Local Position=@n5
   Object Bounds Min Adjustment=@n1
   Object Bounds Max Adjustment=@n2
   Object Bounds Center Adjustment=@n3
   Customization Mask A=@n24
   Customization Mask B=@n10
   IsStatsB=NamedRerouteUsage()
   Side Project Vehicle=@n12
   isCharacter=@n23
   isVehicle=@n16
   isMech=@n6
  ) {materialFunction=/Game/Weapons/WeaponSkins/Wraps/MaterialFunctions/MF_TopSideProjection_Wrap}
  MaterialFunctionCall!(
   isCharacter=NamedRerouteUsage()
   Material=@n27.Result
   ForwardAxis=@n4
   Material_Bypass=@n19
   isMech=@n6
   isVehicle=@n16
   Pre-Skinned Local Normal=@n7
   Local Position=@n5
   Object Bounds Min Adjustment=@n1
   Object Bounds Max Adjustment=@n2
   Object Bounds Center Adjustment=@n3
   "Customization Mask (_CM B)"=@n24
   IsStatsB=NamedRerouteUsage()
  ) {materialFunction=/Game/Weapons/WeaponSkins/Wraps/MaterialFunctions/MF_ActivatedRunes_Wrap}
  n147 = MaterialFunctionCall!(
   Pre-SkinLocalPos=@n5
   IsCharacter=NamedRerouteUsage()
   IsStatsB=NamedRerouteUsage()
   In=@n38.Result
   Pre-SkinLocalNormal=@n7
   Pre-SkinBounds=@n28
   CMask=@n24
   IsVehicle=@n16
   IsMech=@n6
  ) {materialFunction=/Game/Weapons/WeaponSkins/Wraps/MaterialFunctions/MF_BannerWrap}
  MaterialFunctionCall!(
   isMech=@n6
   Material=MaterialFunctionCall!(
    Side Project Vehicle=@n12
    In=MaterialFunctionCall!(Alpha Mask=@n24, GMAP=@n11, In=@n147.Result) {materialFunction=&/MF_JacketWrap}.Result
    ForwardAxis=@n4
    isMech=@n6
    Local Normals=@n7
    Local Position=@n5
    isVehicle=@n16
    Object Bounds Center Adjustment=@n3
    Object Bounds Min Adjustment=@n1
    Object Bounds Max Adjustment=@n2
    "Customization Mask (_CM B)"=@n24
   ) {materialFunction=/Game/Weapons/WeaponSkins/Wraps/MaterialFunctions/MF_Carnaval}.Result
   Side Project Vehicle=@n12
   Local Normals=@n7
   Local Position=@n5
   isVehicle=@n16
   isCharacter=NamedRerouteUsage()
   Object Bounds Min Adjustment=@n1
   Object Bounds Center Adjustment=@n3
   Object Bounds Max Adjustment=@n2
   Customization Mask B=@n24
   "Customization Mask (_CM B)"=@n33
  ) {materialFunction=/Game/Weapons/WeaponSkins/Wraps/MaterialFunctions/MF_CarnavalCycle}
  MaterialFunctionCall!(
   Material=MaterialFunctionCall!(Material=@n31.Result, isVehicle=@n16, ForwardAxis=@n4, Pre-Skinned Local Position=@n5) {materialFunction=/Game/Weapons/WeaponSkins/Wraps/MaterialFunctions/MF_Customization_Holographic}.Result
   Wrap ScaleMultiplier=@n20
   ForwardAxis=@n4
   "Customization Mask (_CM B)"=@n24
   Pre-Skinned Bounds=@n18
   Pre-Skinned Local Position=@n5
   Pre-Skinned Local Normal=@n7
  ) {materialFunction=/Game/Weapons/WeaponSkins/Wraps/MaterialFunctions/MF_Customization_Shatter}
  n148 = MaterialFunctionCall!(
   Material=@n39.Result
   Local Position=@n5
   Local Normals=@n7
   Side Project Vehicle=@n12
   ForwardAxis=@n4
   isVehicle=@n16
   isMech=@n6
   Object Bounds Center Adjustment=@n3
   Customization Mask Primary=@n24
   Customization Mask Curvatur=@n30
   Object Bounds Min Adjustment=@n1
   Object Bounds Max Adjustment=@n2
  ) {materialFunction=/Game/Weapons/WeaponSkins/Wraps/MaterialFunctions/MF_Spectrum}
  n149 = NamedRerouteUsage()
  n150 = NamedRerouteUsage()
  n151 = NamedRerouteUsage()
  n152 = MaterialFunctionCall!(
   ForwardAxis=@n4
   Material=MaterialFunctionCall!(
    Material=MaterialFunctionCall!(
     ObjectBounds_Center=@n3
     Mask=@n33
     PreskionnedPosition=@n5
     isVehicle=@n16
     isMech=@n6
     PreskinnedNormal=@n7
     Material=@n148.Result
     ObjectBounds_Min=@n1
     ObjectBounds_Max=@n2
     Vehicle Side=@n149
     isCharacter=@n150
     IsStatsB=@n151
    ) {materialFunction=/Game/Weapons/WeaponSkins/Wraps/MaterialFunctions/MF_TargetPractice_Wrap}.Result
    isMech=@n6
    isVehicle=@n16
    Pre-Skinned Local Normal=@n7
    Local Position=@n5
    Object Bounds Min Adjustment=@n1
    Object Bounds Max Adjustment=@n2
    Object Bounds Center Adjustment=@n3
    Customization Mask Accent=@n10
    "Customization Mask (_CM B)"=@n24
   ) {materialFunction=/Game/Weapons/WeaponSkins/Wraps/MaterialFunctions/MF_PeeledFoil}.Result
   isMech=@n6
   isVehicle=@n16
   Pre-Skinned Local Normal=@n7
   Local Position=@n5
   Object Bounds Min Adjustment=@n1
   Object Bounds Max Adjustment=@n2
   Object Bounds Center Adjustment=@n3
   "Customization Mask (_CM B)"=@n24
  ) {materialFunction=/Game/Weapons/WeaponSkins/Wraps/MaterialFunctions/MF_NFWrap}
  n153 = MaterialFunctionCall!(
   Material=MaterialFunctionCall!(
    Material=MaterialFunctionCall!(
     Material=@n152.Result
     isMech=@n6
     isVehicle=@n16
     Pre-Skinned Local Normal=@n7
     Local Position=@n5
     Object Bounds Min Adjustment=@n1
     Object Bounds Max Adjustment=@n2
     Object Bounds Center Adjustment=@n3
     "Customization Mask (_CM B)"=@n24
    ) {materialFunction=/Game/Weapons/WeaponSkins/Wraps/MaterialFunctions/MF_OceanRider_Wrap}.Result
    isMech=@n6
    isVehicle=@n16
    Pre-Skinned Local Normal=@n7
    Local Position=@n5
    Object Bounds Min Adjustment=@n1
    Object Bounds Max Adjustment=@n2
    Object Bounds Center Adjustment=@n3
    "Customization Crevasse (_CM G)"=@n30
    "Customization Mask (_CM B)"=@n24
   ) {materialFunction=/Game/Weapons/WeaponSkins/Wraps/MaterialFunctions/MF_GlowVortex_Wrap}.Result
   Pre-Skinned Local Normal=@n7
   Local Position=@n5
   ForwardAxis=@n4
   Cal_Bounds=@n28
   isCharacter=NamedRerouteUsage()
   Side Project Vehicle=@n12
   isMech=@n6
   isVehicle=@n16
   "Customization Mask (_CM B)"=@n24
  ) {materialFunction=/Game/Weapons/WeaponSkins/Wraps/MaterialFunctions/MF_WrapRainbowGlitter}
  n154 = NamedRerouteUsage()
  n155 = NamedRerouteUsage()
  n156 = NamedRerouteUsage()
  MaterialFunctionCall!(
   PreSkin_Bounds=@n28
   Side Project Vehicle=@n12
   isMech=@n6
   In=MaterialFunctionCall!(
    Material=MaterialFunctionCall!(
     Material_input=@n153.Result
     Crevice=@n30
     isMech=@n6
     Side Project Vehicle=@n12
     isVehicle=@n16
     isCharacter=@n154
     Customization Mask=@n24
     Object Bounds Min Adjustment=@n1
     Object Bounds Max Adjustment=@n2
     Object-SafeZone-Virtual-CH-Switch=@n155
     Object-SafeZone-Virtual-Pivot=@n156
    ) {materialFunction=/Game/Weapons/WeaponSkins/Wraps/MaterialFunctions/MF_WrapNeonPulse}.Result
    isMech=@n6
    isVehicle=@n16
    Pre-Skinned Local Normal=@n7
    Local Position=@n5
    Object Bounds Min Adjustment=@n1
    Object Bounds Max Adjustment=@n2
    Object Bounds Center Adjustment=@n3
    "Customization Mask (_CM B)"=@n24
   ) {materialFunction=/Game/Weapons/WeaponSkins/Wraps/MaterialFunctions/MF_GlowCamo_Wrap}.Result
   isVehicle=@n16
   "Customization Mask (_CM B)"=@n24
  ) {materialFunction=/Game/Weapons/WeaponSkins/Wraps/MaterialFunctions/MF_Ultravioletwrap.MF_UltravioletWrap}
  n157 = MaterialFunctionCall!(
   Local Position=@n5
   isMech=@n6
   isVehicle=@n16
   Object Bounds Max Adjustment=@n3
   Object Bounds Min Adjustment=@n2
   Object Bounds Center Adjustment=@n3
   In=MaterialFunctionCall!(In=@n35.Result, Mask=@n24, isVehicle=@n16, isMech=@n6) {materialFunction=/Game/Characters/Player/Male/Medium/Bodies/M_MED_Mastermind/Materials/MF_TwistedGoo}.Result
   Alpha=@n24
   Local normal=@n7
   Side Project Vehicle=@n12
   PreSkin_Bounds=@n28
   Bounds_Extents=@n17.Bound Extents
  ) {materialFunction=/Game/Weapons/WeaponSkins/Wraps/MaterialFunctions/MF_NeonAnimalWrap}
  MaterialFunctionCall!(
   Material=MaterialFunctionCall!(
    Material=MaterialFunctionCall!(Mask=@n33, isMech=@n6, isVehicle=@n16, Material=@n157.Result) {materialFunction=/Game/Weapons/WeaponSkins/Wraps/MaterialFunctions/MF_IceBreakerWrap}.Result
    Side Project Vehicle=@n12
    IsMech=@n6
    IsVehicle=@n16
    Wrap Bounds=@n28
    "Customization Crevasse (_CM G)"=@n30
    IsStatsB=NamedRerouteUsage()
    Object-SafeZone-VP-Character-PreView-Switch=NamedRerouteUsage()
    Object-SafeZone-Virtual-Pivot=NamedRerouteUsage()
   ) {materialFunction=/Game/Weapons/WeaponSkins/Wraps/MaterialFunctions/MF_ConstellationWrap}.Result
   isCharacter=NamedRerouteUsage()
   Local Position=@n5
   IsMech=@n6
   IsVehicle=@n16
   "Customization Crevasse (_CM G)"=@n30
   IsStatsB=NamedRerouteUsage()
  ) {materialFunction=/Game/Weapons/WeaponSkins/Wraps/MaterialFunctions/MF_StarsWrap}
  MaterialFunctionCall!(
   Material=@n40.Result
   Side Project Vehicle=@n12
   Warp Mask
   IsMech=@n6
   IsVehicle=@n16
   Wrap Bounds=@n28
   "Customization Crevasse (_CM G)"=@n30
  ) {materialFunction=/Game/Weapons/WeaponSkins/Wraps/MaterialFunctions/MF_Tapas_Wrap}
  n158 = MaterialFunctionCall!(
   Material=@n41.Result_1
   Side Project Vehicle=@n12
   IsCharacter=NamedRerouteUsage()
   isMech=@n6
   isVehicle=@n16
   Calculated Bounds=@n28
   Pre-Skinned Local Normal=@n7
   Local Position=@n5
   Object Bounds Min Adjustment=@n1
   Wrap Bounds
   Object Bounds Max Adjustment=@n2
   Object Bounds Center Adjustment=@n3
   "Customization Mask 2 (_CM B Top Right)"=@n24
   isCharAccent=NamedRerouteUsage()
  ) {materialFunction=/Game/Weapons/WeaponSkins/Wraps/MaterialFunctions/MF_HoneyDew_Wrap}
  n159 = MaterialFunctionCall!(
   Material=MaterialFunctionCall!(
    Material=MaterialFunctionCall!(
     Material=@n158.Result
     isVehicle=@n16
     isMech=@n6
     Side Project Vehicle=@n12
     ForwardAxis=@n4
     Local Position=@n5
     Object Bounds Min Adjustment=@n1
     Object Bounds Max Adjustment=@n2
     Object Bounds Center Adjustment=@n3
     Pre-Skinned Local Normal=@n7
     "Customization Mask (_CM B)"=@n24
    ) {materialFunction=/Game/Weapons/WeaponSkins/Wraps/MaterialFunctions/MF_Date_Wrap}.Result
    Side Project Vehicle=@n12
    isMech=@n6
    isVehicle=@n16
    Pre-Skinned Local Normal=@n7
    Calcated_Bounds=@n28
    Customization Mask Curvature=@n30
    "Customization Mask (_CM B)"=@n24
   ) {materialFunction=/Game/Weapons/WeaponSkins/Wraps/MaterialFunctions/MF_Wrap_Subwoofer}.Result
   isVehicle=@n16
   isMech=@n6
   IsCharacter=NamedRerouteUsage()
   Side Project Vehicle=@n12
   ForwardAxis=@n4
   Local Position=@n5
   Object Bounds Min Adjustment=@n1
   Object Bounds Max Adjustment=@n2
   Object Bounds Center Adjustment=@n3
   Pre-Skinned Local Normal=@n7
   "Customization Mask (_CM B)"=@n24
   "Customization Mask 2 (_CM B Top Right)"=@n10
   isCharAccent=NamedRerouteUsage()
  ) {materialFunction=/Game/Weapons/WeaponSkins/Wraps/MaterialFunctions/MF_Tomato_Wrap}
  n160 = NamedRerouteUsage()
  n161 = MaterialFunctionCall!(
   Material=MaterialFunctionCall!(
    Material=MaterialFunctionCall!(Material=@n159.Result, isVehicle=@n16, Pre-Skinned Local Position=@n5, IsStatsB=@n160) {materialFunction=/Game/Weapons/WeaponSkins/Wraps/MaterialFunctions/MF_Soy_Wrap}.Result
    ForwardAxis=@n4
    IsCharacter=NamedRerouteUsage()
    isMech=@n6
    isVehicle=@n16
    Pre-Skinned Local Normal=@n7
    Calcated_Bounds=@n28
    "Customization Mask (_CM B)"=@n24
    Object Bounds Max Adjustment=@n2
    Object Bounds Min Adjustment=@n1
    Object-SafeZone-Virtual-CH-Switch=NamedRerouteUsage()
    Object-SafeZone-Virtual-Pivot=NamedRerouteUsage()
   ) {materialFunction=/Game/Weapons/WeaponSkins/Wraps/MaterialFunctions/MF_Wrap_Squash}.Result
   ForwardAxis=@n4
   Side Project Vehicle=@n12
   isMech=@n6
   isVehicle=@n16
   Pre-Skinned Local Normal=@n7
   Calcated_Bounds=@n28
   "Customization Mask (_CM B)"=@n24
  ) {materialFunction=/Game/Weapons/WeaponSkins/Wraps/MaterialFunctions/MF_Wrap_NeonJellyFish}
  n162 = NamedRerouteUsage()
  MaterialFunctionCall!(
   Material=MaterialFunctionCall!(
    Material=MaterialFunctionCall!(
     Material=@n161.Result
     isCharacter=@n162
     isMech=@n6
     isVehicle=@n16
     Pre-Skinned Local Normal=@n7
     Calcated_Bounds=@n28
     "Customization Mask (_CM B Top left)"=@n24
     "Customization Mask 2 (_CM B Top Right)"=@n10
    ) {materialFunction=/Game/Weapons/WeaponSkins/Wraps/MaterialFunctions/MF_NeonSign_Wrap}.Result
    isMech=@n6
    isVehicleSide=@n12
    isVehicle=@n16
    ForwardAxis=@n4
    Pre-Skinned Local Normal=@n7
    Calculated_Bounds=@n28
    "Customization Mask (_CM B Top left)"=@n24
    "Customization Mask 2 (_CM B Top Right)"=@n10
   ) {materialFunction=/Game/Weapons/WeaponSkins/Wraps/MaterialFunctions/MF_Birthday2020_Wrap}.Result
   Side Project Vehicle=@n12
   ForwardAxis=@n4
   isMech=@n6
   isVehicle=@n16
   Pre-Skinned Local Normal=@n7
   Calcated_Bounds=@n28
   "Customization Mask (_CM B)"=@n24
  ) {materialFunction=/Game/Weapons/WeaponSkins/Wraps/MaterialFunctions/MF_Wrap_Muzak}
  MaterialFunctionCall!(
   Material=@n42.Result
   ForwardAxis=@n4
   isMech=@n6
   isVehicle=@n16
   Side Project Vehicle=@n12
   Pre-Skinned Local Normal=@n7
   Local Position=@n5
   Object Bounds Min Adjustment=@n1
   Object Bounds Max Adjustment=@n2
   Object Bounds Center Adjustment=@n3
   "Customization Mask (_CM B)"=@n24
  ) {materialFunction=/Game/Weapons/WeaponSkins/Wraps/MaterialFunctions/MF_KevinCouture_Wrap}
  MaterialFunctionCall!(
   Material=@n43.Result
   Side Project Vehicle=@n12
   isMech=@n6
   isVehicle=@n16
   Pre-Skinned Local Normal=@n7
   Calcated_Bounds=@n28
   "Customization Mask (_CM B)"=@n24
  ) {materialFunction=/Game/Weapons/WeaponSkins/Wraps/MaterialFunctions/MF_Wrap_Chip}
  MaterialFunctionCall!(
   Material=@n44.Result_1
   Pre-Skinned Local Position=@n5
   ForwardAxis=@n4
   isState=NamedRerouteUsage()
   Pre-Skinned Local Normal=@n7
   Calcated_Bounds=@n28
   "Customization Mask (_CM B Top left)"=@n24
   "Customization Mask 2 (_CM B Top Right)"=@n10
   isCharAccent=NamedRerouteUsage()
  ) {materialFunction=/Game/Weapons/WeaponSkins/Wraps/MaterialFunctions/MF_Wrap_CosmicPulse}
  MaterialFunctionCall!(
   Material=@n45.Result
   IsState=NamedRerouteUsage()
   Pre-Skinned Local Normal=@n7
   Calcated_Bounds=@n28
   Foward Axis=@n4
   "Customization Mask (_CM B Top left)"=@n24
   "Customization Mask 2 (_CM B Top Right)"=@n10
   isCharAccent=NamedRerouteUsage()
  ) {materialFunction=/Game/Weapons/WeaponSkins/Wraps/MaterialFunctions/MF_Wrap_Shapeshifter}
  MaterialFunctionCall!(
   Material=MaterialFunctionCall!(
    Material=MaterialFunctionCall!(
     Material=@n46.Result_1
     Side Project Vehicle=@n12
     isMech=@n6
     isVehicle=@n16
     Pre-Skinned Local Normal=@n7
     Calcated_Bounds=@n28
     "Customization Mask (_CM B Top left)"=@n24
     "Customization Mask 2 (_CM B Top Right)"=@n10
    ) {materialFunction=/Game/Weapons/WeaponSkins/Wraps/MaterialFunctions/MF_Wrap_Spacefighter}.Result_1
    ForwardAxis=@n4
    isState=NamedRerouteUsage()
    Pre-Skinned Local Normal=@n7
    Calcated_Bounds=@n28
    "Customization Mask (_CM B Top left)"=@n24
    "Customization Mask 2 (_CM B Top Right)"=@n10
    isCharAccent=NamedRerouteUsage()
   ) {materialFunction=/Game/Weapons/WeaponSkins/Wraps/MaterialFunctions/MF_Wrap_NeonXmasSign}.Result
   isState=NamedRerouteUsage()
   Pre-Skinned Local Normal=@n7
   Calcated_Bounds=@n28
   Curavture=@n30
   Customization Mask B=@n24
   "Customization Mask (_CM B)"=@n33
   isCharAccent=NamedRerouteUsage()
  ) {materialFunction=/Game/Weapons/WeaponSkins/Wraps/MaterialFunctions/MF_WRap_ConfettiLights}
  MaterialFunctionCall!(
   Material=MaterialFunctionCall!(
    Material=@n47.Result
    IsState=NamedRerouteUsage()
    Pre-Skinned Local Normal=@n7
    Calcated_Bounds=@n28
    Curavture=@n30
    Customization Mask B=@n24
    "Customization Mask (_CM B)"=@n33
    isCharAccent=NamedRerouteUsage()
   ) {materialFunction=/Game/Weapons/WeaponSkins/Wraps/MaterialFunctions/MF_Wrap_SmashedGlass}.Result
   IsState=NamedRerouteUsage()
   Forward Axis=@n4
   Pre-Skinned Local Normal=@n7
   Calcated_Bounds=@n28
   Customization Mask A=@n33
   Customization Mask B=@n24
   Curavture=@n30
   isCharAccent=NamedRerouteUsage()
  ) {materialFunction=/Game/Weapons/WeaponSkins/Wraps/MaterialFunctions/MF_Wrap_HunterD}
  MaterialFunctionCall!(
   Material=@n32.Result
   Material_Root
   isState
   isStateB
   Pre-Skinned Local Normal=@n7
   Calcated_Bounds=@n28
   Customization Mask Primary=@n24
   Customization Mask Secondary
   Mask Crevise
   Object-SafeZone-Back=NamedRerouteUsage()
   Object-SafeZone-Middle=NamedRerouteUsage()
   Object-SafeZone-Front=NamedRerouteUsage()
  ) {materialFunction=/Game/Weapons/WeaponSkins/Wraps/MaterialFunctions/MF_Function_Left_Right_Decal}
  MaterialFunctionCall!(
   Material=MaterialFunctionCall!(
    Material=@n48.Result_1
    Side Project Vehicle=@n12
    ForwardAxis=@n4
    isMech=@n6
    isVehicle=@n16
    Pre-Skinned Local Normal=@n7
    Calcated_Bounds=@n28
    Curavture=@n30
    "Customization Mask (_CM B)_1"=@n10
    "Customization Mask (_CM B)"=@n24
   ) {materialFunction=/Game/Weapons/WeaponSkins/Wraps/MaterialFunctions/MF_Wrap_StarAces}.Result
   IsState=NamedRerouteUsage()
   Pre-Skinned Local Normal=@n7
   Calcated_Bounds=@n28
   "Customization Mask (_CM B Top left)"=@n10
   "Customization Mask 2 (_CM B Top Right)"=@n24
   isCharAccent=NamedRerouteUsage()
  ) {materialFunction=/Game/Weapons/WeaponSkins/Wraps/MaterialFunctions/MF_Wrap_WarriorC}
  n163 = NamedRerouteUsage()
  n164 = NamedRerouteUsage()
  n165 = NamedRerouteUsage()
  n166 = FunctionInput() {bUsePreviewValueAsDefault=true, inputName=Object-SafeZone-Middle, inputType=FunctionInput_Vector4, sortPriority=101}
  n167 = FunctionInput() {bUsePreviewValueAsDefault=true, inputName=Object-SafeZone-Front, inputType=FunctionInput_Vector4, sortPriority=102}
  n168 = FunctionInput() {bUsePreviewValueAsDefault=true, inputName=Object-SafeZone-Back, inputType=FunctionInput_Vector4, sortPriority=100}
  MaterialFunctionCall!(
   Material=MaterialFunctionCall!(
    Material=MaterialFunctionCall!(
     Material=@n49.Result_1
     ForwardAxis=@n4
     Is_States=@n163
     Pre-Skinned Local Normal=@n7
     Calcated_Bounds=@n28
     "Customization Mask (_CM B)"=@n24
     Curavture=@n30
     Object Bounds Center Adjustment=@n3
     Object Bounds Min Adjustment=@n1
     Object Bounds Max Adjustment=@n2
     Object-SafeZone-Virtual-Pivot=@n164
     Object-SafeZone-Virtual-CH-Switch=@n165
    ) {materialFunction=/Game/Weapons/WeaponSkins/Wraps/MaterialFunctions/MF_Wrap_TurboC}.Result
    ForwardAxis=@n4
    isState=NamedRerouteUsage()
    Pre-Skinned Local Normal=@n7
    Calcated_Bounds=@n28
    Curavture=@n30
    Customization Mask A=@n24
    Customization Mask B=@n10
    Object-SafeZone-Back=@n168
    Object-SafeZone-Middle=@n166
    Object-SafeZone-Front=@n167
   ) {materialFunction=/Game/Weapons/WeaponSkins/Wraps/MaterialFunctions/MF_Wrap_Mustang}.Result
   isState=NamedRerouteUsage()
   Pre-Skinned Local Normal=@n7
   Calcated_Bounds=@n28
   "Customization Mask (_CM B Top left)"=@n10
   "Customization Mask 2 (_CM B Top Right)"=@n24
   isCharAccent=NamedRerouteUsage()
  ) {materialFunction=/Game/Weapons/WeaponSkins/Wraps/MaterialFunctions/MF_Wrap_Yogurt}
  NamedRerouteDeclaration(None=@n168) {name=SafeZone-Back, nodeColor=<rgba 1 0.6588 0.0581 1>}
  NamedRerouteDeclaration(None=@n166) {name=SafeZone-Middle, nodeColor=<rgba 1 0.6588 0.0581 1>}
  NamedRerouteDeclaration(None=FunctionInput() {bUsePreviewValueAsDefault=true, inputName=IsStatesB, inputType=FunctionInput_Vector2}) {name=Is_StatesB, desc="R= Backpack  B=Pickaxe", nodeColor=<rgba 0 0.9253 1 1>}
  NamedRerouteDeclaration(None=@n167) {name=SafeZone-Front, nodeColor=<rgba 1 0.6588 0.0581 1>}
  NamedRerouteDeclaration(None=@n9.B) {name=Mask_Secondary, nodeColor=<rgba 0.0094 0.0094 0.901 1>}
  NamedRerouteDeclaration(None=@n2) {name=Bounds_Max, nodeColor=<rgba 1 0 0.4811 1>}
  NamedRerouteDeclaration(None=@n1) {name=Bounds_Min, nodeColor=<rgba 1 0 0.4811 1>}
  n169 = FunctionInput() {bUsePreviewValueAsDefault=true, inputName=Object-SafeZone-VP_Charcter_Preview_Switch, inputType=FunctionInput_Scalar, sortPriority=105}
  NamedRerouteDeclaration(None=@n169) {name=SafeZone-VP_Charcter_Preview_Switch, nodeColor=<rgba 1 0.6588 0.0581 1>}
  NamedRerouteDeclaration(None=@n4) {name=Forward Axis, nodeColor=<rgba 0.7813 0 1 1>}
  NamedRerouteDeclaration(None=@n89) {name=Is_StatsB_V4, nodeColor=<rgba 0 0.9253 1 1>}
  n170 = FunctionInput() {bUsePreviewValueAsDefault=true, inputName=Object-SafeZone-Virtual-Pivot, inputType=FunctionInput_Vector4, sortPriority=105}
  NamedRerouteDeclaration(None=@n170) {name=SafeZone-Virtual-Pivot, nodeColor=<rgba 1 0.6588 0.0581 1>}
  NamedRerouteDeclaration(None=@n8) {name=Msk_RGBA, nodeColor=<rgba 0 0.4141 1 1>}
  NamedRerouteDeclaration(None=@n87) {name=Is_States, desc="R=vehicle G=VehicleSide =R=Mech A=Character", nodeColor=<rgba 0 0.9253 1 1>}
  NamedRerouteUsage()
  NamedRerouteUsage()
  NamedRerouteUsage()
  NamedRerouteUsage()
  Reroute(None=@n18)
  Reroute(None=@n20)
  Reroute(None=@n16)
  Reroute(None=@n5)
  Reroute(None=@n30)
  Reroute(None=@n28)
  Reroute(None=@n12)
  Reroute(None=@n7)
  Reroute(None=@n6)
  Reroute(None=@n42.Result)
 ]}
