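# Material function for weapon wraps that layers a simple accent material onto the input material, masked by the CM texture and gated by a static switch.
# Accent color comes from flat or blended colors, curve atlases or projected textures (side or top-side), with curvature-driven pit and peak tints, plus roughness, normal and emissive controls.
# A parallel parameter set supports time-of-day blending.
# State inputs select tiling, rotation and masking per item type, and platform switches trim costlier features.
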MaterialFunction /Game/Weapons/WeaponSkins/Wraps/MaterialFunctions/MF_Customization_Simple_V2
graph {nodes [
  n1 = NamedRerouteUsage()
  n2 = ScalarParameter() {group=Customization_Accent, parameterName="AccentV2_BaseColor ---->Emissive", sliderMax=1}
  n3 = NamedRerouteUsage()
  n4 = TextureSample(UVs=NamedRerouteUsage(), Tex=NamedRerouteUsage())
  n5 = NamedRerouteUsage()
  n6 = VectorParameter() {group=Customization_Accent, parameterName=AccentV2_BaseColor_A, defaultValue=<rgba 0.3906 0 0 1>}
  n7 = VectorParameter() {group=Customization_Accent, parameterName=AccentV2_BaseColor_B, defaultValue=<rgba 0.0053 0.0021 0.3958 1>}
  n8 = NamedRerouteUsage()
  n9 = VectorParameter() {group=Customization_Accent, parameterName=AccentV2_BaseColor_C, defaultValue=<rgba 0.076 0.3958 0.0491 1>}
  n10 = LinearInterpolate(
   A=LinearInterpolate(A=@n6.RGB, B=@n7.RGB, Alpha=@n8)
   B=MaterialFunctionCall(A=@n6.RGB, B=@n7.RGB, C=@n9.RGB, Alpha=@n8) {materialFunction=&3ColorBlend}.Result
   Alpha=ScalarParameter() {group=Customization_Accent, parameterName="AccentV2_BaseColor 3 Colors  1=Active", sliderMax=1}
  )
  n11 = VectorParameter() {group=Customization_Accent, parameterName=AccentV2_BaseColor_, defaultValue=<rgba 0.474 0.474 0.474 1>}
  n12 = StaticSwitchParameter(
   True=NamedRerouteUsage()
   False=StaticSwitchParameter(
    True=StaticSwitch(True=@n3, False=@n4.RGB, Value=@n5)
    False=StaticSwitchParameter(True=@n10, False=@n11.RGB) {group=Customization_Accent, parameterName="AccentV2_BaseColor 2\\3_Color"}
   ) {group=Customization_Accent, parameterName=AccentV2_BaseColor-Texture}
  ) {group=Customization_Accent, parameterName=AccentV2_BaseColor CC}
  n13 = VectorParameter() {
   group="Customization_Accent Pits\\Peaks"
   parameterName="AccentV2_BaseColor_Pits_ColorTint   A=0=Active"
   channelNames {a="0=Active", b="", g="", r=""}
   defaultValue=<rgba 1 1 1 1>
  }
  n14 = VectorParameter() {
   group="Customization_Accent Pits\\Peaks"
   parameterName="AccentV2_BaseColor_Peaks_ColorTint   A=0=Active"
   channelNames {a="0=Active", b="", g="", r=""}
   defaultValue=<rgba 1 1 1 1>
  }
  n15 = NamedRerouteUsage()
  n16 = LinearInterpolate(A=Multiply(A=@n14, B=@n12), B=@n12, Alpha=Max(A=@n14."0=Active", B=OneMinus(None=@n15)))
  n17 = NamedRerouteUsage()
  n18 = LinearInterpolate(A=Multiply(A=@n13, B=@n16), B=@n16, Alpha=Max(A=@n13."0=Active", B=OneMinus(None=@n17)))
  n19 = MaterialFunctionCall!("Xb1,PS4"=@n18, "PC-High&&Epic,XSX,PS5"=@n18, "PC-Perf,Android"=@n12, Switch=@n12, "PC-Low(NotPerf)&&Medium"=@n12) {materialFunction=/Game/Effects/Niagara/MasterMaterials/Functions/Common/PlatformSwitch_3D}
  n20 = NamedRerouteUsage()
  n21 = NamedRerouteUsage()
  n22 = StaticSwitch(True=LinearInterpolate(A=@n19.Result, B=NamedRerouteUsage(), Alpha=@n20), False=@n19.Result, Value=@n21)
  n23 = NamedRerouteUsage()
  n24 = StaticSwitchParameter(
   True=LinearInterpolate(
    A=ScalarParameter() {group=Customization_Accent, parameterName=AccentV2_Roughness Low, sliderMax=1}
    B=ScalarParameter() {defaultValue=1, group=Customization_Accent, parameterName=AccentV2_Roughness High, sliderMax=1}
    Alpha=StaticSwitch(True=@n23, False=@n4.A, Value=@n5)
   )
   False=ScalarParameter() {defaultValue=0.5, group=Customization_Accent, parameterName=AccentV2_Roughness, sliderMax=1}
  ) {group=Customization_Accent, parameterName=AccentV2_BaseColor-Texture}
  n25 = NamedRerouteUsage()
  n26 = ScalarParameter() {group=Customization_Accent, parameterName=AccentV2_Metalness, sliderMax=1}
  n27 = ScalarParameter() {group=Customization_Accent_TOD, parameterName=AccentV2_TOD_Metalness, sliderMax=1}
  n28 = NamedRerouteUsage()
  n29 = NamedRerouteUsage()
  n30 = ScalarParameter() {defaultValue=0.5, group=Customization_Accent, parameterName=AccentV2_Specular, sliderMax=1}
  n31 = ScalarParameter() {defaultValue=0.5, group=Customization_Accent_TOD, parameterName=AccentV2_TOD_Specular, sliderMax=1}
  n32 = NamedRerouteUsage()
  n33 = GetMaterialAttributes(None=NamedRerouteUsage()) {attributeGetTypes [#1]}
  n34 = NamedRerouteUsage()
  n35 = SetMaterialAttributes!(
   MaterialAttributes=NamedRerouteUsage()
   BaseColor=LinearInterpolate(A=@n22, Alpha=@n2) {constB=0}
   Roughness=StaticSwitch(True=LinearInterpolate(A=@n24, B=@n25, Alpha=@n20), False=@n24, Value=@n21)
   Metallic=StaticSwitch(True=LinearInterpolate(A=@n26, B=@n27, Alpha=@n28), False=@n26, Value=@n29)
   Specular=StaticSwitch(True=LinearInterpolate(A=@n30, B=@n31, Alpha=@n28), False=@n30, Value=@n29)
   EmissiveColor=LinearInterpolate(B=@n22, Alpha=@n2)
   Normal=StaticSwitch(True=LinearInterpolate(A=@n32, B=@n33.Normal, Alpha=@n34), False=@n32, Value=NamedRerouteUsage())
  ) {attributeSetTypes [
    #2
    #3
    #4
    #5
    #6
    #1
   ]}
  n36 = FunctionInput(Preview=Constant() {r=1}) {bUsePreviewValueAsDefault=true, inputName="Msk_Ts01<--->Ts02", inputType=FunctionInput_Scalar}
  n37 = VectorParameter() {
   group=Customization_Accent
   parameterName="AccentV2__Mask_TS01<-->TS02 Blend"
   channelNames {a="", b="", g="1=Invertes Mask", r="0=Activate"}
   defaultValue=<rgba 1 1 0 1>
  }
  n38 = LinearInterpolate(A=@n36, B=OneMinus(None=@n36), Alpha=@n37."1=Invertes Mask")
  n39 = Max(A=@n38, B=@n37."0=Activate")
  n40 = ScalarParameter() {defaultValue=1, group=Customization_Accent, parameterName="AccentV2__BackPack Mask Disable 0=active 1=off", sliderMax=1}
  n41 = MaterialFunctionCall(Float4=NamedRerouteUsage()) {materialFunction=&BreakOutFloat4Components}
  n42 = NamedRerouteDeclaration(None=@n41.A) {name=isCharAccent, nodeColor=<rgba 0 0.9253 1 1>}
  n43 = OneMinus(None=NamedRerouteUsage())
  n44 = ScalarParameter() {group=Customization_Accent, parameterName="AccentV2_Unmasked_Areas 1=on", sliderMax=1, sortPriority=0}
  n45 = Saturate(None=Max(A=Min(A=@n43, B=@n44), B=NamedRerouteUsage()))
  n46 = Multiply(A=@n45, B=Saturate(None=Add(A=@n40, B=@n42)))
  n47 = ScalarParameter() {defaultValue=1, group=Customization_Accent, parameterName="AccentV2__Pickaxe Mask Disable 0=active 1=off", sliderMax=1}
  n48 = Saturate(None=LinearInterpolate(A=LinearInterpolate(A=@n45, B=@n46, Alpha=@n41.R), B=Multiply(A=@n45, B=@n47), Alpha=@n41.G))
  n49 = Multiply(A=@n39, B=@n48)
  n50 = StaticBoolParameter() {defaultValue=true, group="TopSide_V02_Layer 01 Blend  02", parameterName="TopS Blend 01<--->02"}
  n51 = StaticSwitchParameter(
   True=BlendMaterialAttributes(A=@n1, B=@n35, Alpha=StaticSwitch(True=@n49, False=@n48, Value=@n50)) {vertexAttributeBlendType=UseA}
   False=@n1
  ) {
   desc=Uses the Top Right corner of the B channel of the _CM Mask to define a simple secondary material
   group=Customization
   parameterName="Allow Accent Customization Area (Simple)"
   sortPriority=-499
  }
  FunctionOutput(None=@n51)
  GetMaterialAttributes(None=NamedRerouteUsage()) {attributeGetTypes [#1]}
  n52 = TextureObjectParameter() {
   group=Customization_Accent
   parameterName=AccentV2_TOD_Normal-Texture
   samplerType=SAMPLERTYPE_Normal
   texture=&FlatNormalBlue
  }
  n53 = NamedRerouteUsage()
  n54 = NamedRerouteUsage()
  n55 = NamedRerouteUsage()
  n56 = NamedRerouteUsage()
  n57 = NamedRerouteUsage()
  n58 = ScalarParameter() {defaultValue=1, group=Customization_Accent, parameterName=AccentV2_Alt_Contrast}
  n59 = NamedRerouteUsage()
  MaterialFunctionCall!(
   Texture_Normals=NamedRerouteDeclaration(None=@n52) {name=N_B, nodeColor=<rgba 0 1 0.6016 1>}
   Rotation_Axis=@n53
   Caculated_Bounds=@n54
   Pre_Skinned_Normals=@n55
   IsStatsB=@n56
   IsState=@n57
   Blend_Contrast=@n58
   Fill_alternate_Axis_Opacity
   Scale=@n59
   Offset_Pivot
   Offset
  ) {materialFunction=&/MF_TopsSideProjection_Normals_V2}
  n60 = VectorParameter() {group=Customization_Accent, parameterName="AccentV2_Alt_Axies_Color   A=Roughness", defaultValue=<rgba 0 0 0 1>}
  MaterialFunctionCall!(
   Texture=NamedRerouteDeclaration(None=TextureObjectParameter() {group=Customization_Accent, parameterName=AccentV2_TOD_BaseColor-Tex, texture=&T_BlacksRGB}) {name=BC_B, nodeColor=<rgba 0 1 0.6016 1>}
   Caculated_Bounds=@n54
   Pre_Skinned_Normals=@n55
   Blend_Contrast=@n58
   IsStatsB=@n56
   IsState=@n57
   Fill_alternate_Axis_Opacity
   "Fill_alternate_Axis_Color_A=Roughness"=AppendVector(A=@n60.RGB, B=@n60.RGB)
   Offset
   Scale=@n59
   Offset_Pivot
   Rotation_Axis=@n53
  ) {materialFunction=&/MF_TopsSideProjection_Base_V2}
  MaterialFunctionCall(Float2=NamedRerouteUsage()) {materialFunction=&BreakOutFloat2Components}
  Min(A=@n39, B=@n48)
  Multiply(A=@n36) {constB=-1}
  Multiply(A=MaterialFunctionCall!(X range Y range, Gradient Height, ShowGrid?, Thickness, Color, "f(x)"=@n38) {materialFunction=/Engine/Functions/Engine_MaterialFunctions01/Debug/PlotFunctionOnGraph}."f(x) = y")
  NamedRerouteDeclaration(None=FunctionInput() {inputName=Material, inputType=FunctionInput_MaterialAttributes, sortPriority=-200}) {name=Mat, nodeColor=<rgba 1 0.0703 0 1>}
  n61 = ComponentMask(None=NamedRerouteUsage()) {r=true}
  n62 = Power(
   Base=LinearInterpolate(A=@n61, B=Frac(None=@n61), Alpha=ScalarParameter() {group=Customization_Accent, parameterName="AccentV2_Grad_Loop  1=Active", sliderMax=1})
   Exp=ScalarParameter() {defaultValue=1, group=Customization_Accent, parameterName=AccentV2_Grad_Falloff, sliderMax=1}
  )
  n63 = CurveAtlasRowParameter(CurveTime=@n62) {
   atlas=/Game/Weapons/WeaponSkins/Wraps/Textures/CruveAtlas_Wraps_01
   curve=/Game/Weapons/WeaponSkins/Wraps/Textures/0_ColorCurves/CC_Spectum
   defaultValue=98
   group=Customization_Accent
   parameterName=AccentV2_BaseColor_CC
  }
  NamedRerouteDeclaration(None=MaterialFunctionCall!(
    Index=ScalarParameter() {defaultValue=98, group=Customization_Accent, parameterName="AccentV2_BaseColor_CC  Index", sliderMax=1}
    CCinput=@n63
    CCinput_Alpha
    UV=@n62
   ) {materialFunction=&/MF_Global_CC_Tex}.3V) {name=ColorCurve, nodeColor=<rgba 0 0.4141 1 1>}
  n64 = NamedRerouteUsage()
  n65 = GetMaterialAttributes(None=NamedRerouteUsage()) {attributeGetTypes [#1]}
  n66 = StaticSwitch(
   True=LinearInterpolate(
    A=GetMaterialAttributes(None=@n64) {attributeGetTypes [#1]}.Normal
    B=@n65.Normal
    Alpha=NamedRerouteUsage()
   )
   False=@n65.Normal
   Value=StaticBoolParameter() {defaultValue=true, group="TopSide_V02_Layer 01 Blend  02", parameterName="TopS Blend 01<--->02"}
  )
  n67 = NamedRerouteUsage()
  n68 = TextureObjectParameter() {
   group=Customization_Accent
   parameterName=AccentV2_Normal-Texture
   samplerType=SAMPLERTYPE_Normal
   texture=&FlatNormalBlue
  }
  n69 = MaterialFunctionCall!(
   Normal_Map=@n68
   Offset Rt
   Pivot
   Caculated_Bounds=NamedRerouteUsage()
   Pre_Skinned_Normals=NamedRerouteUsage()
   is_State=NamedRerouteUsage()
   UV_Rotation=NamedRerouteUsage()
   Uvs_Offset=Multiply(A=Time(), B=VectorParameter() {group=Customization_Accent_UV, parameterName=AccentV2_Panning Speed, defaultValue=<rgba 0 0 0 1>}.RGB)
   Uvs_TileRate=NamedRerouteUsage()
   Uvs_Offset_Pivot
   For_Safe_Zones
   Rotation_Comp
   Clamp Y
   Clamp X
   Clamp Padding_Comp
   No_Rotation=StaticBoolParameter() {group=Customization_Accent_UV, parameterName=AccentV2 No_Rotation}
  ) {materialFunction=&/MF_Projection_Side_Normal_V2}
  n70 = NamedRerouteUsage()
  n71 = LinearInterpolate(
   A=@n66
   B=Add(A=StaticSwitch(True=@n67, False=@n69.Normal, Value=@n70), B=@n66)
   Alpha=ScalarParameter() {defaultValue=1, group=Customization_Accent, parameterName=AccentV2_Normal_Strength, sliderMax=1}
  )
  n72 = StaticSwitchParameter(
   True=MaterialFunctionCall("Xb1,PS4"=@n71, "PC-High&&Epic,XSX,PS5"=@n71, "PC-Perf,Android"=@n66, Switch=@n71, "PC-Low(NotPerf)&&Medium"=@n71) {materialFunction=&PlatformSwitch_3D}.Result
   False=@n71
  ) {defaultValue=true, group=Customization_Accent, parameterName=AccentV2_Normal-Android Disable}
  NamedRerouteDeclaration(None=StaticSwitchParameter(True=StaticSwitchParameter(True=@n72, False=@n66) {group=Customization_Accent, parameterName=AccentV2_Normal--Texture}, False=@n66) {group=Customization_Accent, parameterName=AccentV2_BaseColor-Texture}) {name=Normals, nodeColor=<rgba 0.5234 1 0 1>}
  n73 = ScalarParameter() {defaultValue=1, group=Customization_Accent_UV, parameterName=AccentV2_TileRate_Weapon, sortPriority=150}
  n74 = MaterialFunctionCall(Float4=NamedRerouteUsage()) {materialFunction=&BreakOutFloat4Components}
  n75 = LinearInterpolate(
   A=ScalarParameter() {defaultValue=1, group=Customization_Accent_UV, parameterName=AccentV2_TileRate_Vehicle, sortPriority=150}
   B=ScalarParameter() {defaultValue=1, group=Customization_Accent_UV, parameterName=AccentV2_TileRate_Vehicle_Side, sortPriority=150}
   Alpha=@n74.G
  )
  n76 = ScalarParameter() {defaultValue=1, group=Customization_Accent_UV, parameterName=AccentV2_TileRate_Mech, sortPriority=150}
  n77 = ScalarParameter() {defaultValue=1, group=Customization_Accent_UV, parameterName=AccentV2_TileRate_Character, sortPriority=150}
  n78 = MaterialFunctionCall(Float2=NamedRerouteUsage()) {materialFunction=&BreakOutFloat2Components}
  n79 = LinearInterpolate(
   A=LinearInterpolate(A=LinearInterpolate(A=@n73, B=@n75, Alpha=@n74.R), B=LinearInterpolate(A=@n76, B=@n77, Alpha=@n74.A), Alpha=@n74.B)
   B=ScalarParameter() {defaultValue=1, group=Customization_Accent_UV, parameterName=AccentV2_TileRate_Backpack, sortPriority=150}
   Alpha=@n78.R
  )
  n80 = ScalarParameter() {defaultValue=1, group=Customization_Accent_UV, parameterName=AccentV2_TileRate_Pickaxe, sortPriority=150}
  NamedRerouteDeclaration(None=Multiply(
    A=ScalarParameter() {defaultValue=1, group=Customization_Accent_UV, parameterName=AccentV2_TileRate__Master, sortPriority=150}
    B=LinearInterpolate(A=@n79, B=@n80, Alpha=@n78.G)
   )) {name=UVs_TileRate, nodeColor=<rgba 0 0.7188 1 1>}
  n81 = TextureObjectParameter() {group=Customization_Accent, parameterName=AccentV2_BaseColor-Tex, texture=&T_BlacksRGB}
  n82 = NamedRerouteUsage()
  n83 = NamedRerouteUsage()
  n84 = ScalarParameter() {defaultValue=1, group=Customization_Accent, parameterName=AccentV2_Alt_Contrast}
  n85 = NamedRerouteUsage()
  n86 = NamedRerouteUsage()
  n87 = ScalarParameter() {group=Customization_Accent, parameterName="AccentV2_Alt_Axies     1=Use"}
  n88 = VectorParameter() {group=Customization_Accent, parameterName="AccentV2_Alt_Axies_Color   A=Roughness", defaultValue=<rgba 0 0 0 1>}
  n89 = NamedRerouteUsage()
  n90 = NamedRerouteUsage()
  n91 = MaterialFunctionCall(
   Texture=@n81
   Caculated_Bounds=@n82
   Pre_Skinned_Normals=@n83
   Blend_Contrast=@n84
   IsStatsB=@n85
   IsState=@n86
   Fill_alternate_Axis_Opacity=@n87
   "Fill_alternate_Axis_Color_A=Roughness"=AppendVector(A=@n88.RGB, B=@n88.RGB)
   Scale=@n89
   Rotation_Axis=@n90
  ) {materialFunction=&MF_TopsSideProjection_Base_V2}
  NamedRerouteDeclaration(None=@n91.RGB) {name=TS_BC, nodeColor=<rgba 0.9453 1 0 1>}
  NamedRerouteDeclaration(None=@n91.Alpha) {name=TS_R, nodeColor=<rgba 0.9453 1 0 1>}
  NamedRerouteDeclaration(None=MaterialFunctionCall(
    Texture_Normals=@n68
    Rotation_Axis=@n90
    Caculated_Bounds=@n82
    Pre_Skinned_Normals=@n83
    IsStatsB=@n85
    IsState=@n86
    Blend_Contrast=@n84
    Fill_alternate_Axis_Opacity=@n87
    Scale=@n89
   ) {materialFunction=&MF_TopsSideProjection_Normals_V2}.Add_Surface_Normal) {name=TS_N, nodeColor=<rgba 0.9453 1 0 1>}
  n92 = ScalarParameter() {group=Customization_Accent_UV, parameterName="AccentV2_Rotation_Weapon\\Pickaxe", sortPriority=150}
  n93 = MaterialFunctionCall(Float4=NamedRerouteUsage()) {materialFunction=&BreakOutFloat4Components}
  n94 = LinearInterpolate(
   A=ScalarParameter() {group=Customization_Accent_UV, parameterName=AccentV2_Rotation_Vehicle, sortPriority=150}
   B=ScalarParameter() {group=Customization_Accent_UV, parameterName=AccentV2_Rotation_Vehicle_Side, sortPriority=150}
   Alpha=@n93.G
  )
  NamedRerouteDeclaration(None=LinearInterpolate(
    A=LinearInterpolate(A=@n92, B=@n94, Alpha=@n93.R)
    B=ScalarParameter() {group=Customization_Accent_UV, parameterName="AccentV2_Rotation_Mech\\Character\\Back", sortPriority=150}
    Alpha=@n93.B
   )) {name=UVs_Rotation_1, nodeColor=<rgba 0 0.7188 1 1>}
  NamedRerouteDeclaration(None=StaticBoolParameter() {group=Customization_Accent_UV, parameterName="AccentV2 Side------->TopSide"}) {name=TS_State, nodeColor=<rgba 1 0.0938 0 1>}
  n95 = NamedRerouteUsage()
  n96 = ComponentMask(None=@n95) {r=true}
  n97 = ComponentMask(None=Frac(None=@n95)) {r=true}
  n98 = Multiply(A=Min(A=@n97, B=OneMinus(None=@n97))) {constB=2}
  n99 = ScalarParameter() {group=Customization_Accent, parameterName="AccentV2_Grad_Loop  1=Active", sliderMax=1}
  NamedRerouteDeclaration(None=Power(
    Base=LinearInterpolate(A=@n96, B=@n98, Alpha=@n99)
    Exp=ScalarParameter() {defaultValue=1, group=Customization_Accent, parameterName=AccentV2_Grad_Falloff, sliderMax=1}
   )) {name=UV_Grad, nodeColor=<rgba 0.9219 0 1 1>}
  NamedRerouteDeclaration(None=@n81) {name=BC_A, nodeColor=<rgba 0 1 0.6016 1>}
  n100 = TextureSample() {samplerType=SAMPLERTYPE_Masks, texture=/Game/Weapons/FORT_Rifles/Textures/T_Scar_CM}
  n101 = MaterialFunctionCall(Float4=FunctionInput(Preview=@n100.RGBA) {inputName=CM_Msk, inputType=FunctionInput_Vector4, sortPriority=4}) {materialFunction=&BreakOutFloat4Components}
  n102 = Add(A=NamedRerouteDeclaration(None=@n101.G) {name=Msk_Curvature, nodeColor=<rgba 0 1 0.7422 1>}) {constB=-0.5}
  n103 = Saturate(None=@n102)
  n104 = ScalarParameter() {defaultValue=1, group="Customization_Accent Pits\\Peaks", parameterName=AccentV2_Peaks_Falloff, sliderMax=1}
  NamedRerouteDeclaration(None=MaterialFunctionCall(
    In=Power(Base=@n103, Exp=@n104)
    Contrast=ScalarParameter() {group="Customization_Accent Pits\\Peaks", parameterName=AccentV2_Peaks_Contrast, sliderMax=1}
   ) {materialFunction=&CheapContrast}.Result) {name=CV_Peaks, nodeColor=<rgba 1 0.2344 0 1>}
  n105 = CurveAtlasRowParameter(CurveTime=@n62) {
   atlas=&CruveAtlas_Wraps_01
   curve=&CC_Spectum
   defaultValue=98
   group=Customization_Accent
   parameterName=AccentV2_TOD_BaseColor_CC
  }
  NamedRerouteDeclaration(None=MaterialFunctionCall(
    Index=ScalarParameter() {defaultValue=98, group=Customization_Accent, parameterName="AccentV2_TOD_BaseColor_CC  Index", sliderMax=1}
    CCinput=@n105
    CCinput_Alpha=@n105
    UV=@n62
   ) {materialFunction=&MF_Global_CC_Tex}.3V) {name=ColorCurve_TOD, nodeColor=<rgba 0 0.4141 1 1>}
  NamedRerouteDeclaration(None=@n100.RGBA) {name=CM_MAsk, nodeColor=<rgba 1 0 0.3047 1>}
  n106 = NamedRerouteUsage()
  n107 = LinearInterpolate(
   A=@n106
   B=OneMinus(None=@n106)
   Alpha=ScalarParameter() {group=Customization_Accent_TOD, parameterName="AccentV2_Blend_USE TS01--->TS02 Blend Msk 1-Invert MskMorning", sortPriority=1}
  )
  n108 = ScalarParameter() {defaultValue=182, group=Customization_Accent_TOD, parameterName="AccentV2_Blend_TOD_CC  Index"}
  n109 = ScalarParameter() {group=00-FrontEnd-Overrides, parameterName=00-FE-TODCC-Override}
  n110 = CurveAtlasRowParameter(CurveTime=Add(
    A=CollectionParameter() {collection=&FortniteMaterialParameters, parameterName=TimeOfDay_0-1}
    B=@n109
   )) {
   atlas=&CruveAtlas_Wraps_01
   curve=/Game/Weapons/WeaponSkins/Wraps/Textures/0_ColorCurves/Curve_TimeOfDay
   defaultValue=182
   group=Customization_Accent_TOD
   parameterName=AccentV2_Blend_TOD_CC
  }
  n111 = ScalarParameter() {group=Customization_Accent_TOD, parameterName=AccentV2_Blend_TOD_Day}
  n112 = LinearInterpolate(A=@n109, B=Multiply(A=@n109) {constB=-1}, Alpha=@n111)
  n113 = StaticSwitchParameter(
   True=ChannelMaskParameter(None=MaterialFunctionCall(Index=@n108, CCinput=@n110, CCinput_Alpha=@n110) {materialFunction=&MF_Global_CC_Tex}.4v) {group=Customization_Accent_TOD, parameterName=AccentV2_Blend_TOD_CC_Channel}.R
   False=MaterialFunctionCall(
    Value_Morning=ScalarParameter() {defaultValue=0.5, group=Customization_Accent_TOD, parameterName=AccentV2_Blend_TOD_Morning}
    Value_Day=Add(A=@n112, B=@n111)
    Value_Evening=ScalarParameter() {defaultValue=0.5, group=Customization_Accent_TOD, parameterName=AccentV2_Blend_TOD_Evening}
    Value_Night=ScalarParameter() {defaultValue=1, group=Customization_Accent_TOD, parameterName=AccentV2_Blend_TOD_Night}
   ) {materialFunction=&Scalar_Time_of_Day}.Result
  ) {group=Customization_Accent_TOD, parameterName="AccentV2_Blend_TOD --->CCTOD"}
  n114 = StaticBoolParameter() {
   desc="USe topside Blend Mask to Transistion "
   group=Customization_Accent_TOD
   parameterName="AccentV2_Blend_USE TS01--->TS02 Blend Msk"
   sortPriority=0
  }
  n115 = StaticBoolParameter() {defaultValue=true, group="TopSide_V02_Layer 01 Blend  02", parameterName="TopS Blend 01<--->02"}
  NamedRerouteDeclaration(None=StaticSwitch(True=StaticSwitch(True=@n107, False=@n113, Value=@n114), False=@n113, Value=@n115)) {name=Blend TOD, nodeColor=<rgba 0 0.8594 1 1>}
  n116 = VectorParameter() {group=Customization_Accent_TOD, parameterName=AccentV2_TOD_BaseColor_A, defaultValue=<rgba 0.3906 0 0 1>}
  n117 = VectorParameter() {group=Customization_Accent_TOD, parameterName=AccentV2_TOD_BaseColor_B, defaultValue=<rgba 0.0053 0.0021 0.3958 1>}
  n118 = NamedRerouteUsage()
  n119 = LinearInterpolate(A=@n116.RGB, B=@n117.RGB, Alpha=@n118)
  n120 = VectorParameter() {group=Customization_Accent_TOD, parameterName=AccentV2_TOD_BaseColor_C, defaultValue=<rgba 0.076 0.3958 0.0491 1>}
  n121 = MaterialFunctionCall(A=@n116.RGB, B=@n117.RGB, C=@n120.RGB, Alpha=@n118) {materialFunction=&3ColorBlend}
  n122 = ScalarParameter() {group=Customization_Accent_TOD, parameterName="AccentV2_TOD_BaseColor 3 Colors  1=Active", sliderMax=1}
  n123 = VectorParameter() {group=Customization_Accent_TOD, parameterName=AccentV2_TOD_BaseColor_, defaultValue=<rgba 0.474 0.474 0.474 1>}
  n124 = StaticSwitchParameter(
   True=NamedRerouteUsage()
   False=StaticSwitchParameter(True=LinearInterpolate(A=@n119, B=@n121.Result, Alpha=@n122), False=@n123.RGB) {group=Customization_Accent_TOD, parameterName="AccentV2_TOD_BaseColor 2\\3_Color"}
  ) {group=Customization_Accent_TOD, parameterName=AccentV2_TOD_BaseColor CC}
  n125 = VectorParameter() {
   group=Customization_Accent_TOD
   parameterName="AccentV2_TOD_BaseColor_Pits_ColorTint   A=0=Active"
   channelNames {a="0=Active", b="", g="", r=""}
   defaultValue=<rgba 1 1 1 1>
  }
  n126 = VectorParameter() {
   group=Customization_Accent_TOD
   parameterName="AccentV2_TOD_BaseColor_Peaks_ColorTint   A=0=Active"
   channelNames {a="0=Active", b="", g="", r=""}
   defaultValue=<rgba 1 1 1 1>
  }
  n127 = NamedRerouteUsage()
  n128 = LinearInterpolate(A=Multiply(A=@n126, B=@n124), B=@n124, Alpha=Max(A=@n126."0=Active", B=OneMinus(None=@n127)))
  n129 = NamedRerouteUsage()
  n130 = LinearInterpolate(A=Multiply(A=@n125, B=@n128), B=@n128, Alpha=Max(A=@n125."0=Active", B=OneMinus(None=@n129)))
  NamedRerouteDeclaration(None=MaterialFunctionCall("Xb1,PS4"=@n130, "PC-High&&Epic,XSX,PS5"=@n130, "PC-Perf,Android"=@n124, Switch=@n124, "PC-Low(NotPerf)&&Medium"=@n124) {materialFunction=&PlatformSwitch_3D}.Result) {name=TOD_BC, nodeColor=<rgba 0.3594 1 0 1>}
  n131 = VectorParameter() {group=Customization_Accent_UV, parameterName="AccentV2_RotationTS_Weapon\\PickAxe", defaultValue=<rgba 0 0 -0.25 1>}
  n132 = VectorParameter() {group=Customization_Accent_UV, parameterName=AccentV2_RotationTS_Vehicle_Side, defaultValue=<rgba 0 0 0 1>}
  n133 = MaterialFunctionCall(Float4=NamedRerouteUsage()) {materialFunction=&BreakOutFloat4Components}
  n134 = LinearInterpolate(
   A=VectorParameter() {group=Customization_Accent_UV, parameterName=AccentV2_RotationTS_Vehicle, defaultValue=<rgba 0 0 0 1>}.RGB
   B=@n132.RGB
   Alpha=@n133.G
  )
  n135 = VectorParameter() {group=Customization_Accent_UV, parameterName="AccentV2_RotationTS_Mech\\Character\\Backpack", defaultValue=<rgba 0.25 0.25 0 1>}
  NamedRerouteDeclaration(None=LinearInterpolate(A=LinearInterpolate(A=@n131.RGB, B=@n134, Alpha=@n133.R), B=@n135.RGB, Alpha=@n133.B)) {name=TS_Rotation, nodeColor=<rgba 1 0 0.375 1>}
  n136 = ScalarParameter() {group=Customization_Accent_TOD, parameterName=AccentV2_TOD_Roughness Low, sliderMax=1}
  n137 = ScalarParameter() {defaultValue=1, group=Customization_Accent_TOD, parameterName=AccentV2_TOD_Roughness High, sliderMax=1}
  n138 = TextureSample(UVs=NamedRerouteUsage(), Tex=NamedRerouteUsage())
  n139 = NamedRerouteUsage()
  n140 = StaticSwitch(True=NamedRerouteUsage(), False=@n138.A, Value=@n139)
  NamedRerouteDeclaration(None=StaticSwitchParameter(
    True=LinearInterpolate(A=@n136, B=@n137, Alpha=@n140)
    False=ScalarParameter() {defaultValue=0.5, group=Customization_Accent_TOD, parameterName=AccentV2_TOD_Roughness, sliderMax=1}
   ) {group=Customization_Accent, parameterName=AccentV2_BaseColor-Texture}) {name=TOD_ROughtness, nodeColor=<rgba 0.3594 1 0 1>}
  n141 = StaticBool() {value=true}
  n142 = StaticBoolParameter() {group=Customization_Accent_TOD, parameterName=AccentV2_Blend_TOD, sortPriority=0}
  NamedRerouteDeclaration(None=StaticSwitch(True=StaticSwitch(True=@n141, False=@n142, Value=@n114), False=@n142, Value=@n115)) {name=TOD_State, nodeColor=<rgba 0 1 1 1>}
  NamedRerouteDeclaration(None=@n36) {name="TS01--->Ts02MSk", nodeColor=<rgba 0.8047 0 1 1>}
  NamedRerouteDeclaration(None=@n101.B) {name=Msk_Accent, nodeColor=<rgba 0 1 0.7656 1>}
  NamedRerouteDeclaration(None=FunctionInput() {inputName=Material_Root, inputType=FunctionInput_MaterialAttributes, sortPriority=-200}) {name=Root_Mat, nodeColor=<rgba 1 0.0703 0 1>}
  NamedRerouteDeclaration(None=@n39) {name="TS01--->Ts02_Mask", nodeColor=<rgba 0.8047 0 1 1>}
  NamedRerouteDeclaration(None=@n101.R) {name=Msk_Primary, nodeColor=<rgba 1 0.8906 0 1>}
  NamedRerouteDeclaration(None=FunctionInput(Preview=Constant4Vector().RGBA) {bUsePreviewValueAsDefault=true, inputName=IsState, inputType=FunctionInput_Vector4, sortPriority=4}) {name=IsState, nodeColor=<rgba 0.1497 1 0.1633 1>}
  NamedRerouteDeclaration(None=FunctionInput(Preview=Constant4Vector().RGBA) {bUsePreviewValueAsDefault=true, inputName=IsState_Bv4, inputType=FunctionInput_Vector4, sortPriority=4}) {name=IsState_B, nodeColor=<rgba 0.1497 1 0.1633 1>}
  n143 = PreSkinnedLocalBounds()
  n144 = Subtract(A=VertexInterpolator(VS=PreSkinnedPosition()).PS, B=@n143.Min)
  n145 = MaterialFunctionCall(Color=Subtract(A=@n143.Max, B=@n143.Max)) {materialFunction=&SplitComponents}
  n146 = Max(A=Max(A=@n145.R, B=@n145.R), B=@n145.B)
  NamedRerouteDeclaration(None=FunctionInput(Preview=Divide(A=@n144, B=@n146)) {bUsePreviewValueAsDefault=true, inputName=Calcated_Bounds, sortPriority=11}) {name=Caclcated_Bounds, nodeColor=<rgba 0 0.9297 1 1>}
  n147 = Saturate(None=Multiply(A=@n102) {constB=-1})
  n148 = ScalarParameter() {defaultValue=1, group="Customization_Accent Pits\\Peaks", parameterName=AccentV2_Pits_Falloff, sliderMax=1}
  NamedRerouteDeclaration(None=MaterialFunctionCall(
    In=Power(Base=@n147, Exp=@n148)
    Contrast=ScalarParameter() {group="Customization_Accent Pits\\Peaks", parameterName=AccentV2_Pits_Contrast, sliderMax=1}
   ) {materialFunction=&CheapContrast}.Result) {name=CV_Pits, nodeColor=<rgba 1 0.2344 0 1>}
  n149 = PreSkinnedNormal()
  NamedRerouteDeclaration(None=FunctionInput(Preview=VertexInterpolator(VS=@n149).PS) {bUsePreviewValueAsDefault=true, inputName=Pre-Skinned Local Normal, sortPriority=10}) {name=PreSkin_Normal, nodeColor=<rgba 0.8281 1 0 1>}
  NamedRerouteDeclaration(None=@n69.UVs) {name=Uvs, nodeColor=<rgba 0.6406 1 0 1>}
  Saturate(None=@n49)
  ScalarParameter() {
   defaultValue=1
   group=Customization_Accent
   parameterName="AccentV2__Mask_TS01<-->TS02 Blend invert   1=inverted"
   sliderMax=1
   sortPriority=0
  }
  ScalarParameter() {defaultValue=1, group=Customization_Accent, parameterName="AccentV2__Mask_TS01<-->TS02 Blend Msk  1=off", sliderMax=1, sortPriority=0}
  StaticSwitchParameter(True=StaticSwitch(True=NamedRerouteUsage(), False=@n138.RGB, Value=@n139)) {group=Customization_Accent, parameterName=AccentV2_BaseColor-Texture}
 ]}
