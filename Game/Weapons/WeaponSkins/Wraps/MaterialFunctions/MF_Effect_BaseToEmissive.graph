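# Weapon wrap material function that cross-feeds BaseColor and Emissive on an incoming material, behind the Effect_Base-To-Emissive static switch.
# Two vector parameters' alphas control how much each channel is added into the other, while their RGB darkens or tints the source channel by the same amount.
# The result blends over the original using the CM mask's red (primary) and blue (accent) channels, gated by scalar parameters, with Msk Disable overriding the mask.
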MaterialFunction /Game/Weapons/WeaponSkins/Wraps/MaterialFunctions/MF_Effect_BaseToEmissive
graph {nodes [
  n1 = TextureSample() {samplerType=SAMPLERTYPE_Normal, texture=/Game/Weapons/FORT_Rifles/T_SCAR_N}
  n2 = Constant3Vector() {constant=<rgba 1 0 0 0>}
  n3 = GetMaterialAttributes(None=FunctionInput(Preview=SetMaterialAttributes!(MaterialAttributes, Normal=@n1.RGB, BaseColor=@n2) {attributeSetTypes [#1, #2]}) {inputName=Material, inputType=FunctionInput_MaterialAttributes}) {attributeGetTypes [
    #2
    #3
    #1
    #4
   ]}
  n4 = NamedRerouteDeclaration(None=@n3.MaterialAttributes) {name=Root_Material, nodeColor=<rgba 0.7812 1 0 1>}
  n5 = VectorParameter() {
   group="Effect_BaseColor--->Emissive"
   parameterName="EFBCEM--Base Darken\\Tint  A-->Emissive"
   channelNames {a="--->Emissive", b="", g="", r=""}
  }
  n6 = Multiply(A=@n3.BaseColor, B=@n5.RGB)
  n7 = VectorParameter() {
   group="Effect_BaseColor--->Emissive"
   parameterName="EFBCEM--Emissive Darken\\Tint   A-->BaseColor"
   channelNames {a="--->BaseColor", b="", g="", r=""}
  }
  n8 = Multiply(A=@n3.EmissiveColor, B=@n7.RGB)
  n9 = SetMaterialAttributes!(
   MaterialAttributes=@n4
   BaseColor=Add(A=LinearInterpolate(A=@n3.BaseColor, B=@n6, Alpha=@n5."--->Emissive"), B=LinearInterpolate(B=@n3.EmissiveColor, Alpha=@n7."--->BaseColor"))
   EmissiveColor=Add(
    A=LinearInterpolate(B=@n3.BaseColor, Alpha=@n5."--->Emissive") {constB=0}
    B=LinearInterpolate(A=@n3.EmissiveColor, B=@n8, Alpha=@n7."--->BaseColor") {constB=0}
   )
  ) {attributeSetTypes [#2, #4]}
  n10 = TextureSample() {samplerType=SAMPLERTYPE_Masks, texture=/Game/Weapons/FORT_Rifles/Textures/T_Scar_CM}
  n11 = MaterialFunctionCall(Float4=FunctionInput(Preview=@n10.RGBA) {inputName=CM_Msk, inputType=FunctionInput_Vector4, sortPriority=4}) {materialFunction=&BreakOutFloat4Components}
  n12 = NamedRerouteDeclaration(None=@n11.R) {name=Msk_Prime, nodeColor=<rgba 0 1 0.7422 1>}
  n13 = ScalarParameter() {defaultValue=1, group="Effect_BaseColor--->Emissive", parameterName="EFBCEM--Msk Primary  1=active", sortPriority=0}
  n14 = NamedRerouteDeclaration(None=@n11.B) {name=Msk_Accent, nodeColor=<rgba 0 1 0.7656 1>}
  n15 = ScalarParameter() {group="Effect_BaseColor--->Emissive", parameterName="EFBCEM--Msk Accent   1=active", sortPriority=0}
  n16 = Max(
   A=Max(A=Min(A=@n12, B=@n13), B=Min(A=@n14, B=@n15))
   B=ScalarParameter() {group="Effect_BaseColor--->Emissive", parameterName="EFBCEM--Msk Disable   1=active", sortPriority=0}
  )
  FunctionOutput(None=StaticSwitchParameter(True=BlendMaterialAttributes(A=@n4, B=@n9, Alpha=@n16) {vertexAttributeBlendType=UseA}, False=@n4) {group=Customization, parameterName=Effect_Base-To-Emissive})
  NamedRerouteDeclaration(None=@n11.G) {name=Msk_Curvature, nodeColor=<rgba 0 1 0.7422 1>}
  NamedRerouteDeclaration(None=@n10.RGBA) {name=CM_Mask, nodeColor=<rgba 1 0 0.3047 1>}
  NamedRerouteDeclaration(None=@n3.Normal) {name=Root_Normal, nodeColor=<rgba 0.4062 0 1 1>}
 ]}
