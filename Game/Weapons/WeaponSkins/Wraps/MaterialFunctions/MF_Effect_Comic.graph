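# Material function for weapon wraps that applies an optional comic-book look, masked by primary, accent and decal masks.
# A pre-skinned triplanar pattern and edge term mix ink color with tinted, desaturated base color, optionally routed to emissive.
# Midtone shading thresholds sun-to-normal lighting against a screentone texture on a camera-facing virtual plane anchored to the object or actor.
# Object-state inputs select per-type scales and edge strengths; FOV also adjusts screentone scale.
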MaterialFunction /Game/Weapons/WeaponSkins/Wraps/MaterialFunctions/MF_Effect_Comic
graph {nodes [
  n1 = CollectionParameter() {
   collection=&FortniteMaterialParameters
   parameterName=SunlightDirectionalVector
  }
  n2 = ComponentMask(None=PixelNormalWS()) {b=true, g=true, r=true}
  n3 = LinearInterpolate(
   A=@n2
   B=ComponentMask(None=VertexNormalWS()) {b=true, g=true, r=true}
   Alpha=ScalarParameter() {defaultValue=0.25, group=Effect_Comic_MidTone, parameterName=Eff_Comic_MidTone_Pixel-Vertex}
  )
  n4 = ConstantBiasScale(None=DotProduct(A=ComponentMask(None=@n1) {b=true, g=true, r=true}, B=ShadingPathSwitch(Default=@n3, Forward=@n2, Mobile=@n2)))
  n5 = Constant() {desc=Needs at least a small amount of this or it will break, r=0.25}
  n6 = MaterialFunctionCall() {materialFunction=&CameraDirectionVector}
  n7 = Normalize(VectorInput=Multiply(A=Constant3Vector() {constant=<rgba 1 1 0 0>}, B=@n6.Result))
  n8 = ScalarParameter() {defaultValue=70, group=Effect_Comic_MidTone, parameterName=Eff_Comic_MidTone_CameraVectorMin}
  n9 = ObjectPositionWS()
  n10 = VectorParameter() {group=Effect_Comic_MidTone, parameterName=Eff_Comic_MidTone_ActorPosOffset, defaultValue=<rgba 0 0 80 1>}
  n11 = Add(A=ActorPositionWS(), B=@n10.RGB)
  n12 = MaterialFunctionCall(Float4=NamedRerouteUsage()) {materialFunction=&BreakOutFloat4Components}
  n13 = LinearInterpolate(
   A=ScalarParameter() {defaultValue=1.4, group=Effect_Comic_MidTone, parameterName=Eff_Comic_MidTone_Scale-LowFOV_Weapon, sortPriority=100}
   B=LinearInterpolate(
    A=ScalarParameter() {defaultValue=4.2, group=Effect_Comic_MidTone, parameterName=Eff_Comic_MidTone_Scale-LowFOV_Vehicle, sortPriority=100}
    B=ScalarParameter() {defaultValue=4.2, group=Effect_Comic_MidTone, parameterName=Eff_Comic_MidTone_Scale-LowFOV_VehicleSide, sortPriority=100}
    Alpha=@n12.G
   )
   Alpha=@n12.R
  )
  n14 = LinearInterpolate(
   A=ScalarParameter() {defaultValue=1.4, group=Effect_Comic_MidTone, parameterName=Eff_Comic_MidTone_Scale-LowFOV_Mech, sortPriority=100}
   B=ScalarParameter() {defaultValue=1.4, group=Effect_Comic_MidTone, parameterName=Eff_Comic_MidTone_Scale-LowFOV_Character, sortPriority=100}
   Alpha=@n12.A
  )
  n15 = LinearInterpolate(
   A=ScalarParameter() {defaultValue=1.75, group=Effect_Comic_MidTone, parameterName=Eff_Comic_MidTone_Scale-HighFOV_Weapon, sortPriority=100}
   B=LinearInterpolate(
    A=ScalarParameter() {defaultValue=7, group=Effect_Comic_MidTone, parameterName=Eff_Comic_MidTone_Scale-HighFOV_Vehicle, sortPriority=100}
    B=ScalarParameter() {defaultValue=7, group=Effect_Comic_MidTone, parameterName=Eff_Comic_MidTone_Scale-HighFOV_VehicleSide, sortPriority=100}
    Alpha=@n12.G
   )
   Alpha=@n12.R
  )
  n16 = LinearInterpolate(
   A=ScalarParameter() {defaultValue=3, group=Effect_Comic_MidTone, parameterName=Eff_Comic_MidTone_Scale-HighFOV_Mech, sortPriority=100}
   B=ScalarParameter() {defaultValue=3, group=Effect_Comic_MidTone, parameterName=Eff_Comic_MidTone_Scale-HighFOV_Character, sortPriority=100}
   Alpha=@n12.A
  )
  n17 = MaterialFunctionCall(
   Input=MaterialFunctionCall() {materialFunction=/Engine/Functions/Engine_MaterialFunctions02/Utility/FOV}."FOV "
   Input Low=Constant() {r=0.414}
   Input High=Constant() {r=1.091}
   Target Low=Constant()
   Target High=Constant() {r=1}
  ) {materialFunction=&RemapValueRange}
  n18 = MaterialFunctionCall(
   Plane Normal Axis=Add(A=Multiply(A=@n5, B=@n7), B=@n6.Result)
   Plane Center=Add(
    A=Multiply(A=@n6.Result, B=@n8)
    B=StaticSwitchParameter(True=@n9, False=@n11) {group=Effect_Comic_MidTone, parameterName=Eff_Comic_MidTone_ObjectPos, sortPriority=1}
   )
   UV Scale=LinearInterpolate(A=LinearInterpolate(A=@n13, B=@n14, Alpha=@n12.B), B=LinearInterpolate(A=@n15, B=@n16, Alpha=@n12.B), Alpha=Saturate(None=@n17.Result))
  ) {materialFunction=&VirtualPlaneCoordinates}
  n19 = StaticSwitchParameter(
   True=StaticSwitchParameter(True=NamedRerouteUsage(), False=NamedRerouteUsage()) {group=Effect_Comic_Pattern, parameterName="Eff_Comic_MidTone_Msk__Tex ---->TopS_00B_Sharerd_Tex_Msk"}
   False=NamedRerouteUsage()
  ) {group=Effect_Comic_Pattern, parameterName="Eff_Comic_MidTone_Msk__Tex ---->TopS_00A_Sharerd_Tex_Msk"}
  n20 = Multiply(
   A=ChannelMaskParameter(None=TextureSample(UVs=@n18.UVs, Tex=@n19) {samplerSource=SSM_Wrap_WorldGroupSettings, samplerType=SAMPLERTYPE_LinearGrayscale}.RGBA) {group=Effect_Comic_MidTone, parameterName=Eff_Comic_MidTone_MSk_Channel}.R
   B=ScalarParameter() {defaultValue=0.5, group=Effect_Comic_MidTone, parameterName=Eff_Comic_MidTone_SPread}
  )
  n21 = Saturate(None=Round(None=Subtract(A=@n4, B=@n20)))
  Add(A=@n21, B=ScalarParameter() {defaultValue=0.2, group=Effect_Comic_MidTone, parameterName=Eff_Comic_MidTone_DarksAmount})
  n22 = NamedRerouteUsage()
  n23 = NamedRerouteUsage()
  n24 = FunctionInput() {inputName=Decal_Mask, inputType=FunctionInput_Scalar, sortPriority=4}
  n25 = TextureSample() {samplerType=SAMPLERTYPE_Masks, texture=/Game/Weapons/FORT_Rifles/Textures/T_Scar_CM}
  n26 = MaterialFunctionCall(Float4=FunctionInput(Preview=@n25.RGBA) {inputName=CM_Msk, inputType=FunctionInput_Vector4, sortPriority=4}) {materialFunction=&BreakOutFloat4Components}
  n27 = Min(
   A=ScalarParameter() {defaultValue=1, group=Effect_Comic, parameterName="Eff_Comic_Primary_Msk    1=On", sortPriority=0}
   B=NamedRerouteDeclaration(None=@n26.R) {name=Msk_Prime, nodeColor=<rgba 0 1 0.7422 1>}
  )
  n28 = Min(
   A=NamedRerouteDeclaration(None=@n26.B) {name=Msk_Accent, nodeColor=<rgba 0 1 0.7656 1>}
   B=ScalarParameter() {group=Effect_Comic, parameterName="Eff_Comic_Accent_Msk   1=On", sortPriority=0}
  )
  n29 = Min(
   A=Max(A=ScalarParameter() {defaultValue=1, group=Effect_Comic, parameterName="Eff_Comic_Decal_Msk   1=On", sortPriority=0}, B=OneMinus(None=@n24))
   B=Max(A=Max(A=@n27, B=@n28), B=ScalarParameter() {group=Effect_Comic, parameterName="Eff_Comic_Disable_Masks   1=On", sortPriority=0})
  )
  n30 = TextureSample() {samplerType=SAMPLERTYPE_Normal, texture=/Game/Weapons/FORT_Rifles/T_SCAR_N}
  n31 = Constant3Vector() {constant=<rgba 1 0 0 0>}
  n32 = GetMaterialAttributes(None=FunctionInput(Preview=SetMaterialAttributes!(MaterialAttributes, Normal=@n30.RGB, BaseColor=@n31) {attributeSetTypes [#1, #2]}) {inputName=Material, inputType=FunctionInput_MaterialAttributes}) {attributeGetTypes [
    #2
    #3
    #1
    #4
    #5
   ]}
  FunctionOutput(None=StaticSwitchParameter(
    True=BlendMaterialAttributes(A=@n22, B=@n23, Alpha=@n29) {vertexAttributeBlendType=UseA}
    False=NamedRerouteDeclaration(None=@n32.MaterialAttributes) {name=Root_Material, nodeColor=<rgba 0.7812 1 0 1>}
   ) {group=Customization, parameterName=Effect_Comic})
  n33 = TextureObjectParameter() {
   group=Effect_Comic
   parameterName=Eff_Comic__Msk_Tex
   samplerType=SAMPLERTYPE_Masks
   texture=/Game/Global/Textures/Default/Tileable_Patterns/OakPP_Post
  }
  NamedRerouteDeclaration(None=@n33) {name=Msk_Tex, nodeColor=<rgba 1 0 0.6797 1>}
  NamedRerouteDeclaration(None=@n26.G) {name=Msk_Curvature, nodeColor=<rgba 0 1 0.7422 1>}
  n34 = FunctionInput(Preview=Constant4Vector() {constant=<rgba 0 0 5 -1>}.RGBA) {bUsePreviewValueAsDefault=true, inputName=Object-SafeZone-Virtual-Pivot, inputType=FunctionInput_Vector4, sortPriority=105}
  NamedRerouteDeclaration(None=@n34) {name=SafeZone VP, nodeColor=<rgba 0.5 1 0 1>}
  n35 = ScalarParameter() {defaultValue=148.12, group=Effect_Comic_Pattern_UV, parameterName=Ef_Comic_Pattern_Scale_Vehicle, sortPriority=150}
  n36 = ScalarParameter() {defaultValue=148.11, group=Effect_Comic_Pattern_UV, parameterName=Ef_Comic_Pattern_Scale_Vehicle_Side, sortPriority=150}
  n37 = MaterialFunctionCall(Float4=NamedRerouteUsage()) {materialFunction=&BreakOutFloat4Components}
  n38 = LinearInterpolate(
   A=LinearInterpolate(
    A=ScalarParameter() {defaultValue=32.4262, group=Effect_Comic_Pattern_UV, parameterName=Ef_Comic_Pattern_Scale_Weapon, sortPriority=150}
    B=LinearInterpolate(A=@n35, B=@n36, Alpha=@n37.G)
    Alpha=@n37.R
   )
   B=LinearInterpolate(
    A=ScalarParameter() {defaultValue=64, group=Effect_Comic_Pattern_UV, parameterName=Ef_Comic_Pattern_Scale_Mech, sortPriority=150}
    B=ScalarParameter() {defaultValue=64, group=Effect_Comic_Pattern_UV, parameterName=Ef_Comic_Pattern_Scale_Character, sortPriority=150}
    Alpha=@n37.A
   )
   Alpha=@n37.B
  )
  n39 = ScalarParameter() {defaultValue=64, group=Effect_Comic_Pattern_UV, parameterName=Ef_Comic_Pattern_Scale_Backpack, sortPriority=150}
  n40 = MaterialFunctionCall(Float2=NamedRerouteUsage()) {materialFunction=&BreakOutFloat2Components}
  NamedRerouteDeclaration(None=LinearInterpolate(
    A=LinearInterpolate(A=@n38, B=@n39, Alpha=@n40.R)
    B=ScalarParameter() {defaultValue=32, group=Effect_Comic_Pattern_UV, parameterName=Ef_Comic_Pattern_Scale_Pickaxe, sortPriority=150}
    Alpha=@n40.G
   )) {name=UVs_Scale_Pattern, nodeColor=<rgba 0 0.7188 1 1>}
  n41 = FunctionInput(Preview=Constant()) {bUsePreviewValueAsDefault=true, inputName=Object-SafeZone-VP-Character-PreView-Switch, inputType=FunctionInput_Scalar, sortPriority=105}
  NamedRerouteDeclaration(None=@n41) {name=SafeZone Character-Switch, nodeColor=<rgba 1 0.9375 0 1>}
  NamedRerouteDeclaration(None=@n32.EmissiveColor) {name=Root_Emissive, nodeColor=<rgba 0.4062 0 1 1>}
  n42 = VectorParameter() {
   group=Effect_Comic_Pattern
   parameterName="Eff_Comic_Pattern\\Edge Color"
   channelNames {a=OPACITY, b="", g="", r=""}
   defaultValue=<rgba 0 0 0 1>
  }
  n43 = VectorParameter() {group=Effect_Comic, parameterName=Eff_Comic_BaseColor_BlendColor, defaultValue=<rgba 0.7188 0.5978 0.4384 1>}
  n44 = LinearInterpolate(
   A=NamedRerouteUsage()
   B=@n43.RGB
   Alpha=ScalarParameter() {defaultValue=0.05, group=Effect_Comic, parameterName=Eff_Comic_BaseColor_Blend_Transistion}
  )
  n45 = ScalarParameter() {defaultValue=-0.391, group=Effect_Comic, parameterName=Eff_Comic_BaseColor_Saturation}
  n46 = StaticSwitchParameter(True=NamedRerouteUsage(), False=NamedRerouteUsage()) {group=Effect_Comic_Pattern, parameterName="Eff_Comic_Pattern_Msk__Tex ---->TopS_00B_Sharerd_Tex_Msk"}
  n47 = NamedRerouteUsage()
  n48 = PreSkinnedPosition()
  n49 = PreSkinnedNormal()
  n50 = ChannelMaskParameter(None=MaterialFunctionCall(
    Export Float 4=StaticBool() {value=true}
    TextureObject=StaticSwitchParameter(True=@n46, False=@n47) {group=Effect_Comic_Pattern, parameterName="Eff_Comic_Pattern_Msk__Tex ---->TopS_00A_Sharerd_Tex_Msk"}
    TextureSize=NamedRerouteUsage()
    WorldPosition=VertexInterpolator(VS=@n48).PS
    World Space Normal=VertexInterpolator(VS=@n49).PS
   ) {materialFunction=&/WorldAlignedTexture}.XYZ Texture) {group=Effect_Comic_Pattern, parameterName=Eff_Comic_Pattern_Channel}
  n51 = OneMinus(None=@n50.R)
  n52 = ScalarParameter() {group=Effect_Comic_Pattern, parameterName=Eff_Comic_Pattern_Invert}
  n53 = Saturate(None=LinearInterpolate(
    A=ScalarParameter() {defaultValue=-3.8695, group=Effect_Comic_Pattern, parameterName=Eff_Comic_Pattern_Min}
    B=ScalarParameter() {defaultValue=0.9201, group=Effect_Comic_Pattern, parameterName=Eff_Comic_Pattern_Max}
    Alpha=LinearInterpolate(A=@n50.R, B=@n51, Alpha=@n52)
   ))
  n54 = VectorParameter() {
   group=Effect_Comic_Pattern
   parameterName=Eff_Comic_Pattern_EffectByLighting
   channelNames {a=opacity, b="", g=Max, r=Min}
   defaultValue=<rgba 0 1 0 0>
  }
  n55 = Min(A=SmoothStep(Min=@n54.Min, Max=@n54.Min, Value=@n4), B=@n54.opacity)
  n56 = Add(A=NamedRerouteUsage()) {constB=-0.5}
  n57 = OneMinus(None=Multiply(A=NamedRerouteUsage(), B=Saturate(None=@n56)) {constA=12})
  n58 = Min(
   A=StaticSwitchParameter(True=LinearInterpolate(A=@n53, Alpha=@n55), False=Constant() {r=1}) {group=Effect_Comic_Pattern, parameterName=Eff_Comic_Pattern, sortPriority=0}
   B=Max(A=Saturate(None=@n57), B=ScalarParameter() {group="Effect_Comic_Pattern\\Edge", parameterName=Eff_Comic_Edge_Opacity, sortPriority=0})
  )
  n59 = NamedRerouteDeclaration(None=LinearInterpolate(A=@n42, B=Desaturation(None=@n44, Fraction=@n45), Alpha=Min(A=@n42.OPACITY, B=@n58))) {name=BaseColor, nodeColor=<rgba 1 0 0.3047 1>}
  n60 = VectorParameter() {
   group=Effect_Comic_MidTone
   parameterName=Eff_Comic_MidTone_ColorTint
   channelNames {a=Strength, b="", g="", r=""}
   defaultValue=<rgba 0 0 0 0.2>
  }
  n61 = LinearInterpolate(A=@n60, Alpha=Max(A=@n21, B=@n60.Strength)) {constA=1}
  n62 = StaticSwitchParameter(True=Multiply(A=@n59, B=Saturate(None=@n61)), False=@n59) {group=Effect_Comic_MidTone, parameterName=Eff_Comic_MidTones, sortPriority=0}
  n63 = ScalarParameter() {group=Effect_Comic, parameterName="Eff_Comic__Basecolor---->Emissive"}
  n64 = LinearInterpolate(B=@n62, Alpha=@n63)
  n65 = NamedRerouteUsage()
  NamedRerouteDeclaration(None=SetMaterialAttributes!(MaterialAttributes=NamedRerouteUsage(), BaseColor=LinearInterpolate(A=@n62, Alpha=@n63) {constB=0}, EmissiveColor=Add(A=@n64, B=@n65)) {attributeSetTypes [#2, #5]}) {name=Material, nodeColor=<rgba 0.1016 0 1 1>}
  n66 = TextureObjectParameter() {
   group=TopSide_V02
   parameterName=TopS_00B_Sharerd_Tex_Msk
   samplerType=SAMPLERTYPE_Masks
   sortPriority=249
   texture=&T_BlackMask
  }
  NamedRerouteDeclaration(None=@n66) {name=Tex_00B_Shared_Tex, nodeColor=<rgba 1 0.8906 0 1>}
  n67 = ScalarParameter() {defaultValue=10, group="Effect_Comic_Pattern\\Edge", parameterName=Ef_Comic_Edge_Strength_Vehicle}
  n68 = ScalarParameter() {defaultValue=10, group=Effect_Comic_Pattern_UV, parameterName=Ef_Comic_Edge_Strength_Vehicle_Side}
  n69 = MaterialFunctionCall(Float4=NamedRerouteUsage()) {materialFunction=&BreakOutFloat4Components}
  n70 = LinearInterpolate(
   A=LinearInterpolate(
    A=ScalarParameter() {defaultValue=10, group="Effect_Comic_Pattern\\Edge", parameterName=Ef_Comic_Edge_Strength_Weapon}
    B=LinearInterpolate(A=@n67, B=@n68, Alpha=@n69.G)
    Alpha=@n69.R
   )
   B=LinearInterpolate(
    A=ScalarParameter() {defaultValue=10, group="Effect_Comic_Pattern\\Edge", parameterName=Ef_Comic_Edge_Strength_Mech}
    B=ScalarParameter() {defaultValue=10, group="Effect_Comic_Pattern\\Edge", parameterName=Ef_Comic_Edge_Strength_Character}
    Alpha=@n69.A
   )
   Alpha=@n69.B
  )
  n71 = ScalarParameter() {defaultValue=10, group="Effect_Comic_Pattern\\Edge", parameterName=Ef_Comic_Edge_Strength_Backpack}
  n72 = MaterialFunctionCall(Float2=NamedRerouteUsage()) {materialFunction=&BreakOutFloat2Components}
  NamedRerouteDeclaration(None=LinearInterpolate(
    A=LinearInterpolate(A=@n70, B=@n71, Alpha=@n72.R)
    B=ScalarParameter() {defaultValue=10, group="Effect_Comic_Pattern\\Edge", parameterName=Ef_Comic_Edge_Strength_Pickaxe}
    Alpha=@n72.G
   )) {name=Edge_Strength, nodeColor=<rgba 0 0.7188 1 1>}
  n73 = TextureObjectParameter() {
   group=TopSide_V02
   parameterName=TopS_00A_Sharerd_Tex_Msk
   samplerType=SAMPLERTYPE_Masks
   sortPriority=249
   texture=&T_BlackMask
  }
  NamedRerouteDeclaration(None=@n73) {name=Tex_00A_Shared_Tex, nodeColor=<rgba 1 0.8906 0 1>}
  NamedRerouteDeclaration(None=@n25.RGBA) {name=CM_Mask, nodeColor=<rgba 1 0 0.3047 1>}
  NamedRerouteDeclaration(None=FunctionInput(Preview=Constant4Vector().RGBA) {bUsePreviewValueAsDefault=true, inputName=IsState, inputType=FunctionInput_Vector4, sortPriority=4}) {name=IsState, nodeColor=<rgba 0.1497 1 0.1633 1>}
  NamedRerouteDeclaration(None=FunctionInput(Preview=Constant2Vector()) {bUsePreviewValueAsDefault=true, inputName=IsState_B, inputType=FunctionInput_Vector2, sortPriority=4}) {name=IsState_B, nodeColor=<rgba 0.1497 1 0.1633 1>}
  n74 = PreSkinnedLocalBounds()
  n75 = Subtract(A=VertexInterpolator(VS=PreSkinnedPosition()).PS, B=@n74.Min)
  n76 = MaterialFunctionCall(Color=Subtract(A=@n74.Max, B=@n74.Max)) {materialFunction=&SplitComponents}
  n77 = Max(A=Max(A=@n76.R, B=@n76.R), B=@n76.B)
  NamedRerouteDeclaration(None=FunctionInput(Preview=Divide(A=@n75, B=@n77)) {bUsePreviewValueAsDefault=true, inputName=Calcated_Bounds, sortPriority=11}) {name=Caclcated_Bounds, nodeColor=<rgba 0 0.9297 1 1>}
  NamedRerouteDeclaration(None=@n32.BaseColor) {name=Root_BaseColor, nodeColor=<rgba 0.4062 0 1 1>}
  n78 = PreSkinnedNormal()
  NamedRerouteDeclaration(None=FunctionInput(Preview=VertexInterpolator(VS=@n78).PS) {bUsePreviewValueAsDefault=true, inputName=Pre-Skinned Local Normal, sortPriority=10}) {name=PreSkin_Normal, nodeColor=<rgba 0.8281 1 0 1>}
  NamedRerouteDeclaration(None=@n32.Normal) {name=Root_Normal, nodeColor=<rgba 0.4062 0 1 1>}
  TextureSample() {samplerType=SAMPLERTYPE_Masks, texture=&T_Scar_CM}
 ]}
