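# Material function for weapon wraps that turns masked areas of an input material into a faceted gemstone, behind a static switch.
# Triplanar virtual-box normals feed fake refraction and reflection sky-light samples for a tinted, time-of-day-scaled emissive.
# A custom HLSL height-map raymarch (single sample on low-end platforms) and fresnel caustics from a curve atlas or cubemap add depth.
# Channel, top-side and decal masks control the blend, with tiling per item type.
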MaterialFunction /Game/Weapons/WeaponSkins/Wraps/MaterialFunctions/MF_Effect_GemStone
graph {nodes [
  n1 = NamedRerouteUsage()
  n2 = DeriveNormalZ(InXY=ComponentMask(None=NamedRerouteUsage()) {g=true, r=true})
  n3 = Transform(None=@n2)
  n4 = Multiply(
   A=ComponentMask(None=DotProduct(A=@n1, B=@n3)) {r=true}
   B=ScalarParameter() {defaultValue=2, group=Effect_Gemstone_RayMarching, parameterName=EFGM_GradientMulti}
  )
  n5 = Constant() {r=0.5}
  Add(A=@n4, B=@n5)
  AppendVector(A=@n4, B=@n5)
  AtmosphericLightVector()
  Constant2Vector()
  Constant3Vector() {constant=<rgba 1 0 0 0>}
  Constant4Vector()
  Custom() {code="return View.DirectionalLightDirection;", description=DirectionalLightVector, inputs []}
  n6 = NamedRerouteUsage()
  Divide(B=@n6)
  n7 = NamedRerouteUsage()
  n8 = NamedRerouteUsage()
  Divide(A=@n7, B=@n8)
  n9 = NamedRerouteUsage()
  n10 = NamedRerouteUsage()
  n11 = FunctionInput() {inputName=Decal_Mask, inputType=FunctionInput_Scalar, sortPriority=90}
  n12 = Constant()
  n13 = StaticSwitchParameter(
   True=ChannelMaskParameter(None=FunctionInput() {inputName=Decal_Mask_RGBA, inputType=FunctionInput_Vector4, sortPriority=90}) {group=Effect-Gemstone-Mask, parameterName=EFGM_Decal_Mask_RGBA_Channel}.R
   False=@n12
  ) {group=Effect-Gemstone-Mask, parameterName=EFGM_Decal_Mask_RGBA}
  n14 = OneMinus(None=LinearInterpolate(B=@n11, Alpha=ScalarParameter() {defaultValue=1, group=Effect-Gemstone-Mask, parameterName="EFGM_Decal_Alpha Masked  1=Active"}))
  n15 = Constant() {r=1}
  n16 = FunctionInput() {inputName=TS_Layer02_Mask, inputType=FunctionInput_Vector4, sortPriority=99}
  n17 = Constant() {r=1}
  n18 = StaticSwitchParameter(
   True=ChannelMaskParameter(None=FunctionInput() {inputName=TS_Layer01_Mask, inputType=FunctionInput_Vector4, sortPriority=99}) {group=Effect-Gemstone-Mask, parameterName=EFGM_TS_Layer01_Mask_Channel}.R
   False=StaticSwitchParameter(True=ChannelMaskParameter(None=@n16) {group=Effect-Gemstone-Mask, parameterName=EFGM_TS_Layer02_Mask_Channel}.R, False=@n17) {group=Effect-Gemstone-Mask, parameterName=EFGM_TS_Layer02_Mask}
  ) {group=Effect-Gemstone-Mask, parameterName=EFGM_TS_Layer01_Mask}
  n19 = StaticBoolParameter() {group=TopSide_V02, parameterName=TopSideProjection_V02}
  n20 = ScalarParameter() {defaultValue=1, group=Effect-Gemstone-Mask, parameterName="EFGM_Mask_Primary  1=Active", sortPriority=0}
  n21 = Constant()
  n22 = LinearInterpolate(A=@n21, B=NamedRerouteUsage(), Alpha=ScalarParameter() {group=TopSide_V02, parameterName="TopS_Mask_Accent  0=Off"})
  n23 = ScalarParameter() {group=TopSide_V02, parameterName="TopS_Mask_Primary\\Accent  1=Off"}
  n24 = FunctionInput(Preview=TextureSample() {samplerType=SAMPLERTYPE_Masks}.RGBA) {inputName=CM_Msk, inputType=FunctionInput_Vector4, sortPriority=4}
  n25 = MaterialFunctionCall(Float4=@n24) {materialFunction=&BreakOutFloat4Components}
  n26 = Max(
   A=StaticSwitch(True=Max(A=@n22, B=@n23), False=@n21, Value=StaticBoolParameter() {group=TopSide_V02, parameterName=TopSideProjection_V02})
   B=NamedRerouteDeclaration(None=@n25.R) {name=Msk_Prime, nodeColor=<rgba 0 1 0.7422 1>}
  )
  n27 = Max(
   A=Min(A=StaticSwitch(True=@n18, False=@n17, Value=@n19), B=Min(A=@n20, B=@n26))
   B=Min(
    A=NamedRerouteDeclaration(None=@n25.B) {name=Msk_Accent, nodeColor=<rgba 0 1 0.7656 1>}
    B=ScalarParameter() {group=Effect-Gemstone-Mask, parameterName="EFGM_Mask_Accent  1=Active", sortPriority=0}
   )
  )
  n28 = ScalarParameter() {group=Effect-Gemstone-Mask, parameterName="EFGM_Mask_Disable 1=Active", sortPriority=0}
  n29 = Max(
   A=StaticSwitchParameter(True=StaticSwitchParameter(True=@n11, False=@n13) {group=Effect-Gemstone-Mask, parameterName=EFGM_Decal_Mask_Alpha}, False=@n12) {group=Customization, parameterName=Function_Left_Right_Decal_V2}
   B=Min(A=StaticSwitchParameter(True=@n14, False=@n15) {group=Customization, parameterName=Function_Left_Right_Decal_V2}, B=Max(A=@n27, B=@n28))
  )
  n30 = GetMaterialAttributes(None=FunctionInput(Preview=SetMaterialAttributes!(MaterialAttributes, Normal, BaseColor) {attributeSetTypes [#1, #2]}) {inputName=Material_1, inputType=FunctionInput_MaterialAttributes}) {attributeGetTypes [
    #2
    #3
    #1
    #4
   ]}
  FunctionOutput(None=StaticSwitchParameter(
    True=BlendMaterialAttributes(A=@n9, B=@n10, Alpha=@n29) {vertexAttributeBlendType=UseA}
    False=NamedRerouteDeclaration(None=@n30.MaterialAttributes) {name=Root_Material, nodeColor=<rgba 0.7812 1 0 1>}
   ) {group=Customization, parameterName=Effect_GemStone}) {outputName=Result_1}
  n31 = VectorParameter() {group=Effect_GemStone_Base, parameterName=EFGM_DiamondTint, defaultValue=<rgba 0.8771 0.9768 1 1>}
  n32 = SkyLightEnvMapSample(
   Direction=ReflectionVectorWS(CustomWorldNormal=NamedRerouteUsage())
   Roughness=ScalarParameter() {defaultValue=0.1, group=Effect_GemStone_Base, parameterName=EFGM_ReflectionRoughness}
  )
  n33 = MaterialFunctionCall!(
   Ray Direction=StaticSwitchParameter(True=NamedRerouteUsage(), False=NamedRerouteUsage()) {group=Effect_GemStone_Base, parameterName=EFGM_UseDoubleRefraction}
   Surface Normal=NamedRerouteUsage()
   Refractive Index Origin=NamedRerouteUsage()
   Refractive Index Target=Constant() {r=1}
  ) {materialFunction=/Engine/Functions/Engine_MaterialFunctions01/Vectors/Refract}
  n34 = SkyLightEnvMapSample(Direction=@n33.Vector, Roughness=ScalarParameter() {defaultValue=0.2, group=Effect_GemStone_Base, parameterName=EFGM_RefractionRoughness})
  n35 = Desaturation(None=LinearInterpolate(A=@n32, B=@n34, Alpha=NamedRerouteUsage()), Fraction=Constant() {r=1})
  n36 = LinearInterpolate(
   A=ScalarParameter() {defaultValue=0.05, group=Effect_GemStone_Base, parameterName=EFGM_DiamondMinIntensity}
   Alpha=SmoothStep(
    Min=ScalarParameter() {group=Effect_GemStone_Base, parameterName=EFGM_DiamondSmoothMin}
    Max=ScalarParameter() {defaultValue=1, group=Effect_GemStone_Base, parameterName=EFGM_DiamondSmoothMax}
    Value=Saturate(None=@n35)
   )
  )
  n37 = Multiply(
   A=Multiply(
    A=ScalarParameter() {defaultValue=2, group=Effect_GemStone_Base, parameterName=EFGM_DiamondEmissiveIntensity}
    B=Multiply(A=@n31.RGB, B=@n36)
   )
   B=NamedRerouteUsage()
  )
  Multiply(A=@n37, B=NamedRerouteUsage())
  n38 = GetMaterialAttributes(None=NamedRerouteDeclaration(None=NamedRerouteUsage()) {name=Material_IN, nodeColor=<rgba 0 0.3438 1 1>}) {attributeGetTypes [#2, #1, #5]}
  n39 = VectorParameter() {group=Effect_GemStone_Override, parameterName=EFGM_BaseColorSubtle_Tint, defaultValue=<rgba 1 1 1 1>}
  n40 = StaticSwitchParameter(
   True=VectorParameter() {group=Effect_GemStone_Override, parameterName=EFGM_PureColor_BaseColor, defaultValue=<rgba 1 1 1 1>}.RGB
   False=@n38.BaseColor
  ) {group=Effect_GemStone_Override, parameterName=EFGM_UsePureColorAsBaseColor}
  n41 = Multiply(
   A=Multiply(A=NamedRerouteUsage(), B=NamedRerouteDeclaration(None=@n40) {name=BaseColor_OG, nodeColor=<rgba 0.0078 0 1 1>})
   B=ScalarParameter() {defaultValue=1, group=Effect_GemStone_Override, parameterName=EFGM_BaseColorMulti}
  )
  n42 = NamedRerouteUsage()
  n43 = ScalarParameter() {defaultValue=0.2, group=Effect_GemStone_Override, parameterName=EFGM_DiamondNormal_Flatten}
  n44 = SetMaterialAttributes!(
   MaterialAttributes=@n38.MaterialAttributes
   BaseColor=Saturate(None=Multiply(A=@n39.RGB, B=@n41))
   Metallic=ScalarParameter() {defaultValue=0.75, group=Effect_GemStone_Override, parameterName=EFGM_MetalicOverride}
   EmissiveColor=Add(A=@n38.EmissiveColor, B=NamedRerouteUsage())
   Specular=ScalarParameter() {defaultValue=0.5, group=Effect_GemStone_Override, parameterName=EFGM_Spec_Override}
   Normal=MaterialFunctionCall(
    BaseNormal=NamedRerouteDeclaration(None=@n38.Normal) {name=Normal_OG, nodeColor=<rgba 1 0.3047 0 1>}
    AdditionalNormal=MaterialFunctionCall(Normal=@n42, Flatness=@n43) {materialFunction=&FlattenNormal}.Result
   ) {materialFunction=&BlendAngleCorrectedNormals}.Result
  ) {attributeSetTypes [
    #2
    #4
    #5
    #6
    #1
   ]}
  NamedRerouteDeclaration(None=@n44) {name=Material, nodeColor=<rgba 0 1 0.1094 1>}
  n45 = ScalarParameter() {defaultValue=1, group=Effect_Gemstone_RayMarching_Uv, parameterName=EFGM_NormalTile_Vehicle}
  n46 = ScalarParameter() {defaultValue=1, group=Effect_Gemstone_RayMarching_Uv, parameterName=EFGM_NormalTile_Vehicle_Side}
  n47 = MaterialFunctionCall(Float4=NamedRerouteUsage()) {materialFunction=&BreakOutFloat4Components}
  n48 = LinearInterpolate(
   A=LinearInterpolate(
    A=ScalarParameter() {defaultValue=1, group=Effect_Gemstone_RayMarching_Uv, parameterName=EFGM_NormalTile_Weapon}
    B=LinearInterpolate(A=@n45, B=@n46, Alpha=@n47.G)
    Alpha=@n47.R
   )
   B=LinearInterpolate(
    A=ScalarParameter() {defaultValue=1, group=Effect_Gemstone_RayMarching_Uv, parameterName=EFGM_NormalTile_Mech}
    B=ScalarParameter() {defaultValue=1, group=Effect_Gemstone_RayMarching_Uv, parameterName=EFGM_NormalTile_Character}
    Alpha=@n47.A
   )
   Alpha=@n47.B
  )
  n49 = ScalarParameter() {defaultValue=1, group=Effect_Gemstone_RayMarching_Uv, parameterName=EFGM_NormalTile_Backpack}
  n50 = MaterialFunctionCall(Float2=NamedRerouteUsage()) {materialFunction=&BreakOutFloat2Components}
  NamedRerouteDeclaration(None=LinearInterpolate(
    A=LinearInterpolate(A=@n48, B=@n49, Alpha=@n50.R)
    B=ScalarParameter() {defaultValue=1, group=Effect_Gemstone_RayMarching_Uv, parameterName=EFGM_NormalTile_Pickaxe}
    Alpha=@n50.G
   )) {name=normal_tile, nodeColor=<rgba 0 0.7188 1 1>}
  n51 = ScalarParameter() {defaultValue=1, group=Effect_Gemstone_RayMarching_Uv, parameterName=EFGM_RayMarchingTilling_Vehicle}
  n52 = ScalarParameter() {defaultValue=1, group=Effect_Gemstone_RayMarching_Uv, parameterName=EFGM_RayMarchingTilling_Vehicle_Side}
  n53 = MaterialFunctionCall(Float4=NamedRerouteUsage()) {materialFunction=&BreakOutFloat4Components}
  n54 = LinearInterpolate(
   A=LinearInterpolate(
    A=ScalarParameter() {defaultValue=1, group=Effect_Gemstone_RayMarching_Uv, parameterName=EFGM_RayMarchingTilling_Weapon}
    B=LinearInterpolate(A=@n51, B=@n52, Alpha=@n53.G)
    Alpha=@n53.R
   )
   B=LinearInterpolate(
    A=ScalarParameter() {defaultValue=1, group=Effect_Gemstone_RayMarching_Uv, parameterName=EFGM_RayMarchingTilling_Mech}
    B=ScalarParameter() {defaultValue=1, group=Effect_Gemstone_RayMarching_Uv, parameterName=EFGM_RayMarchingTilling_Character}
    Alpha=@n53.A
   )
   Alpha=@n53.B
  )
  n55 = ScalarParameter() {defaultValue=1, group=Effect_Gemstone_RayMarching_Uv, parameterName=EFGM_RayMarchingTilling_Backpack}
  n56 = MaterialFunctionCall(Float2=NamedRerouteUsage()) {materialFunction=&BreakOutFloat2Components}
  NamedRerouteDeclaration(None=LinearInterpolate(
    A=LinearInterpolate(A=@n54, B=@n55, Alpha=@n56.R)
    B=ScalarParameter() {defaultValue=1, group=Effect_Gemstone_RayMarching_Uv, parameterName=EFGM_RayMarchingTilling_Pickaxe}
    Alpha=@n56.G
   )) {name=ray_tile, nodeColor=<rgba 0 0.7188 1 1>}
  n57 = NamedRerouteUsage()
  n58 = ScalarParameter() {defaultValue=5, group=Effect_GemStone_Base, parameterName=EFGM_BackfaceDepth}
  n59 = ScalarParameter() {defaultValue=1, group=Effect_GemStone_Base, parameterName=EFGM_BackfaceTiling}
  n60 = ScalarParameter() {defaultValue=1, group=Effect_GemStone_Base, parameterName=EFGM_BackfaceBlendContrast}
  n61 = VectorParameter() {group=Effect_GemStone_Base, parameterName=EFGM_BackfaceOffset, defaultValue=<rgba 0 0 50 1>}
  n62 = StaticBool() {value=true}
  NamedRerouteDeclaration(None=Transform(None=MaterialFunctionCall!(Texture=@n57, Depth=@n58, Tile=@n59, BlendContrast=@n60, SpeedXY, ActorOffset=@n61.RGB, bUseLocalCameraVector=@n62, bUseDitherSingleSample) {materialFunction=/Game/Global/MaterialFunctions/MF_TriplanarVirtualBox}.Result)) {name=BackfaceNormal, nodeColor=<rgba 0.1016 1 0 1>}
  n63 = CameraVectorWS()
  n64 = ScalarParameter() {defaultValue=1.5, group=Effect_GemStone_Base, parameterName=EFGM_IOR}
  NamedRerouteDeclaration(None=MaterialFunctionCall(
    Ray Direction=Multiply(A=@n63) {constB=-1}
    Surface Normal=NamedRerouteUsage()
    Refractive Index Origin=Constant() {r=1}
    Refractive Index Target=NamedRerouteDeclaration(None=@n64) {name=IOR, nodeColor=<rgba 0 1 0.9297 1>}
   ) {materialFunction=&Refract}.Vector) {name=RefractionVector, nodeColor=<rgba 1 0 0.4219 1>}
  n65 = Multiply(A=DotProduct(A=NamedRerouteUsage(), B=NamedRerouteUsage())) {constB=-1}
  n66 = Constant() {r=0.73}
  n67 = Constant() {r=0.78}
  NamedRerouteDeclaration(None=LinearInterpolate(Alpha=MaterialFunctionCall(Input=@n65, Input Low=@n66, Input High=@n67) {materialFunction=&RemapValue_0-1}.Result) {constA=0.1, constB=0.9}) {name=TIR, nodeColor=<rgba 0 0.7422 1 1>}
  n68 = VectorParameter() {group=Effect_GemStone_Base, parameterName=EFGM_Normal_Offset_UV}
  n69 = TextureSample(UVs=Add(A=Multiply(A=@n7, B=@n8), B=ComponentMask(None=@n68.RGB) {g=true, r=true}), Tex=NamedRerouteUsage()) {samplerType=SAMPLERTYPE_Normal}
  NamedRerouteDeclaration(None=Transform(None=NamedRerouteDeclaration(None=@n69.RGB) {name=TangentNormal, nodeColor=<rgba 1 0.6094 0 1>})) {name=WorldSurfaceNormal, nodeColor=<rgba 0 1 0.1328 1>}
  n70 = TextureObjectParameter() {
   group=Effect_GemStone_Base
   parameterName=EFGM_DiamondNormalTexture
   samplerType=SAMPLERTYPE_Normal
   texture=&FlatNormalBlue
  }
  NamedRerouteDeclaration(None=@n70) {name=NormalTexture, nodeColor=<rgba 1 0 0.75 1>}
  n71 = Constant4Vector()
  NamedRerouteDeclaration(None=FunctionInput(Preview=@n71.RGBA) {bUsePreviewValueAsDefault=true, inputName=IsState, inputType=FunctionInput_Vector4, sortPriority=4}) {name=IsState, nodeColor=<rgba 0.1497 1 0.1633 1>}
  NamedRerouteDeclaration(None=FunctionInput(Preview=@n71.RGBA) {bUsePreviewValueAsDefault=true, inputName=IsState_B, inputType=FunctionInput_Vector4, sortPriority=4}) {name=IsState_B, nodeColor=<rgba 0.1497 1 0.1633 1>}
  n72 = PreSkinnedLocalBounds()
  n73 = Subtract(A=VertexInterpolator(VS=PreSkinnedPosition()).PS, B=@n72.Min)
  n74 = MaterialFunctionCall(Color=Subtract(A=@n72.Max, B=@n72.Max)) {materialFunction=&SplitComponents}
  n75 = Max(A=Max(A=@n74.R, B=@n74.R), B=@n74.B)
  NamedRerouteDeclaration(None=FunctionInput(Preview=Divide(A=@n73, B=@n75)) {bUsePreviewValueAsDefault=true, inputName=Calcated_Bounds, sortPriority=11}) {name=Caclcated_Bounds, nodeColor=<rgba 0 0.9297 1 1>}
  n76 = PreSkinnedNormal()
  NamedRerouteDeclaration(None=FunctionInput(Preview=VertexInterpolator(VS=@n76).PS) {bUsePreviewValueAsDefault=true, inputName=Pre-Skinned Local Normal, sortPriority=10}) {name=PreSkin_Normal, nodeColor=<rgba 0.8281 1 0 1>}
  n77 = FunctionInput(Preview=Constant4Vector() {constant=<rgba 0 0 5 -1>}.RGBA) {bUsePreviewValueAsDefault=true, inputName=Object-SafeZone-Virtual-Pivot, inputType=FunctionInput_Vector4, sortPriority=105}
  NamedRerouteDeclaration(None=@n77) {name=SafeZone VP, nodeColor=<rgba 0.5 1 0 1>}
  n78 = FunctionInput(Preview=Constant()) {bUsePreviewValueAsDefault=true, inputName=Object-SafeZone-VP-Character-PreView-Switch, inputType=FunctionInput_Scalar, sortPriority=105}
  NamedRerouteDeclaration(None=@n78) {name=SafeZone Character-Switch, nodeColor=<rgba 1 0.9375 0 1>}
  n79 = ScalarParameter() {defaultValue=5, group=Effect_GemStone_Base, parameterName=EFGM_ExponentIn}
  n80 = NamedRerouteUsage()
  NamedRerouteDeclaration(None=OneMinus(None=Fresnel(ExponentIn=@n79, Normal=@n80) {baseReflectFraction=0})) {name=FresnelMask, nodeColor=<rgba 1 0.0469 0 1>}
  NamedRerouteDeclaration(None=@n33.Vector) {name=BackfaceRefraction, nodeColor=<rgba 0.7109 0 1 1>}
  n81 = MaterialFunctionCall(Float3=Abs(None=NamedRerouteUsage())) {materialFunction=&BreakOutFloat3Components}
  n82 = Constant() {r=0.5}
  n83 = NamedRerouteUsage()
  n84 = ComponentMask(None=@n83) {b=true, desc=YZ, g=true}
  n85 = ComponentMask(None=@n83) {desc=YX, g=true, r=true}
  n86 = ComponentMask(None=@n83) {b=true, desc=XZ, r=true}
  n87 = Constant() {r=0.5}
  n88 = If(A=@n81.R, B=@n87, "A > B"=@n84, "A < B"=@n86)
  n89 = Constant() {r=0.5}
  n90 = If(A=@n81.B, B=@n89, "A > B"=@n85, "A < B"=@n84)
  n91 = MaterialFunctionCall(Float4=NamedRerouteUsage()) {materialFunction=&BreakOutFloat4Components}
  n92 = LinearInterpolate(
   A=If(A=@n81.R, B=@n82, "A > B"=@n84, "A < B"=If(A=@n81.B, B=@n82, "A > B"=@n85, "A < B"=@n86))
   B=LinearInterpolate(A=If(A=@n81.B, B=@n87, "A > B"=@n85, "A < B"=@n88), B=If(A=@n81.G, B=@n89, "A > B"=@n86, "A < B"=@n90), Alpha=@n91.G)
   Alpha=@n91.R
  )
  n93 = Constant() {r=0.5}
  n94 = If(A=@n81.G, B=@n93, "A > B"=@n86, "A < B"=If(A=@n81.R, B=@n93, "A > B"=@n84, "A < B"=@n85))
  MaterialExpressionMultiply_57 = Multiply(A=NamedRerouteDeclaration(None=LinearInterpolate(A=@n92, B=@n94, Alpha=@n91.B)) {name=UVS_Cheap_Tri_1, nodeColor=<rgba 1 0 0.8438 1>}, B=@n6)
  MaterialExpressionTransform_4 = Transform(None=CameraVectorWS()) {transformSourceType=TRANSFORMSOURCE_World, transformType=TRANSFORM_Tangent}
  n95 = Constant()
  MaterialExpressionMaterialFunctionCall_5 = MaterialFunctionCall(
   "Xb1,PS4"=Constant() {desc="This low step will brings some artifact around edges, but it is not that noticable", r=1}
   "PC-High&&Epic,XSX,PS5"=Constant() {r=1}
   "PC-Perf,Android"=@n95
   Switch=@n95
   "PC-Low(NotPerf)&&Medium"=Constant() {r=1}
  ) {materialFunction=&PlatformSwitch_1D}
  n96 = ScalarParameter() {defaultValue=0.3, group=Effect_Gemstone_RayMarching, parameterName=EFGM_DiamondRayMarchStepSize, sliderMax=0.2, sortPriority=13}
  n97 = ScalarParameter() {defaultValue=0.1, group=Effect_Gemstone_RayMarching, parameterName=EFGM_DiamondRayMarchStepSize_Perf, sliderMax=0.2, sortPriority=13}
  MaterialExpressionMaterialFunctionCall_6 = MaterialFunctionCall("Xb1,PS4"=@n96, "PC-High&&Epic,XSX,PS5"=@n96, "PC-Perf,Android"=@n97, Switch=@n97, "PC-Low(NotPerf)&&Medium"=@n96) {materialFunction=&PlatformSwitch_1D}
  MaterialExpressionTextureObjectParameter_1 = TextureObjectParameter() {group=Effect_Gemstone_RayMarching, parameterName=EFGM_DiamondHeight, texture=/Game/Global/Textures/Default/Blanks/T_FlatGrey}
  MaterialExpressionNamedRerouteUsage_79 = NamedRerouteUsage()
  n98 = Custom!(
   uv=@MaterialExpressionMultiply_57
   viewDir=@MaterialExpressionTransform_4
   steps=@MaterialExpressionMaterialFunctionCall_5.Result
   stepdist=@MaterialExpressionMaterialFunctionCall_6.Result
   tex=@MaterialExpressionTextureObjectParameter_1
   Ntex=@MaterialExpressionNamedRerouteUsage_79
  ) {
   code="float3 rayStep = viewDir * -1;\r\nfloat4 heightTex = Texture2DSample(tex,texSampler,uv);\r\nfloat4 NormalTex;\r\n\r\n\r\nint Steps = floor(steps);\r\n//float steps = 16;\r\n//float stepdist = float(0.1);\r\n\r\n//[UNROLL]\r\nfor(int i=0; i <= Steps; i++)\r\n{\r\n\tif( max(heightTex.r, 0.0001) <= i  / Steps)\r\n\t{\r\n\t\treturn (NormalTex*i)/Steps; \r\n\t\t\r\n\t}\r\n\r\n\tfloat stepct =  heightTex.r*stepdist;\r\n\t\r\n\tuv += rayStep * stepct;\r\n    if( i == Steps)\r\n    {\r\n        NormalTex = Texture2DSample(Ntex,texSampler,uv.xy);\r\n    }\r\n    else\r\n    {\r\n\t    heightTex = Texture2DSample(tex,texSampler,uv.xy );\r\n        NormalTex = Texture2DSample(Ntex,texSampler,uv.xy);\r\n    }\r\n\t//RayMarchUV = uv.xy + distortStr;\r\n\r\n}\r\n\r\nreturn NormalTex;"
   showCode=true
   inputs [
    {inputName=uv, input {expression=&self:MaterialExpressionMultiply_57}}
    {inputName=viewDir, input {expression=&self:MaterialExpressionTransform_4}}
    {
     inputName=steps
     input {expression=&self:MaterialExpressionMaterialFunctionCall_5}
    }
    {
     inputName=stepdist
     input {expression=&self:MaterialExpressionMaterialFunctionCall_6}
    }
    {
     inputName=tex
     input {expression=&self:MaterialExpressionTextureObjectParameter_1}
    }
    {
     inputName=Ntex
     input {expression=&self:MaterialExpressionNamedRerouteUsage_79}
    }
   ]
  }
  n99 = TextureSample(UVs=@MaterialExpressionMultiply_57, Tex=NamedRerouteUsage())
  NamedRerouteDeclaration(None=MaterialFunctionCall!("Xb1,PS4"=@n98, "PC-High&&Epic,XSX,PS5"=@n98, "PC-Perf,Android"=@n99.RGB, Switch=@n99.RGB, "PC-Low(NotPerf)&&Medium"=@n98) {materialFunction=/Game/Effects/Niagara/MasterMaterials/Functions/Common/PlatformSwitch_3D}.Result) {name=RaymarchDepth, nodeColor=<rgba 0.6641 0 1 1>}
  n100 = Multiply(
   A=Fresnel(ExponentIn=ScalarParameter() {defaultValue=1, group=Effect_Gemstone_RayMarching, parameterName=EFGM_Fresnel_Exp}, Normal=@n3) {baseReflectFraction=0, exponent=2}
   B=ScalarParameter() {defaultValue=2, group=Effect_Gemstone_RayMarching, parameterName=EFGM_SineMulti}
  ) {constB=2}
  n101 = NamedRerouteDeclaration(None=SmoothStep(Value=Sine(None=@n100)) {constMax=0.5, constMin=0.3}) {name=Fresnel_Caustics, nodeColor=<rgba 0.0312 0 1 1>}
  n102 = TextureSampleParameterCube!(UVs=Transform(None=@n2), Bias, Apply View MipBias) {
   group=Effect_Gemstone_RayMarching
   mipValueMode=TMVM_MipBias
   parameterName=EFGM_RefractionCubeMap
   samplerType=SAMPLERTYPE_LinearColor
   texture=/Game/Global/Textures/Cubemaps/TC_Normals_Random_S
  }
  n103 = MaterialFunctionCall(Float3=@n102.RGB) {materialFunction=&BreakOutFloat3Components}
  n104 = MaterialFunctionCall(X=@n103.R, Y=Clamp(None=@n103.G) {maxDefault=0.3}, Z=@n103.R) {materialFunction=&MakeFloat3}
  n105 = ScalarParameter() {defaultValue=-0.3, group=Effect_Gemstone_RayMarching, parameterName=EFGM_Desaturation_Caustics}
  n106 = Multiply(A=@n101, B=Multiply(A=Desaturation(None=@n104.Result, Fraction=@n105), B=NamedRerouteUsage()))
  n107 = ScalarParameter() {defaultValue=167, group=Effect_Gemstone_RayMarching, parameterName="EFGM_DiamondGradient  CC Index"}
  n108 = Frac(None=@n4)
  n109 = CurveAtlasRowParameter(CurveTime=@n108) {
   atlas=/Game/Weapons/WeaponSkins/Wraps/Textures/CruveAtlas_Wraps_01
   curve=/Game/Weapons/WeaponSkins/Wraps/Textures/C_PS_GemColors
   defaultValue=167
   group=Effect_Gemstone_RayMarching
   parameterName=EFGM_DiamondGradient
  }
  n110 = Multiply(
   A=@n101
   B=Desaturation(
    None=MaterialFunctionCall!(Index=@n107, CCinput=@n109, CCinput_Alpha, UV=@n108) {materialFunction=&/MF_Global_CC_Tex}.3V
    Fraction=@n105
   )
  )
  NamedRerouteDeclaration(None=Multiply(
    A=ScalarParameter() {defaultValue=0.4, group=Effect_Gemstone_RayMarching, parameterName=EFGM_CausticsIntensity}
    B=StaticSwitchParameter(True=@n106, False=@n110) {group=Effect_Gemstone_RayMarching, parameterName=EFGM_UseCubeMapForCaustics}
   ) {constA=0.4}) {name=Emissive_Caustics, nodeColor=<rgba 0.5703 1 0 1>}
  n111 = Constant() {r=0.5}
  n112 = Constant() {r=1}
  n113 = Constant() {r=0.3}
  n114 = Constant() {r=0.25}
  n115 = Multiply(A=NamedRerouteUsage(), B=@n37)
  n116 = Add(
   A=NamedRerouteUsage()
   B=Multiply(A=NamedRerouteDeclaration(None=@n115) {name=Emissive_VirtualPlane, nodeColor=<rgba 0 0.2734 1 1>}, B=NamedRerouteUsage())
  )
  n117 = Multiply(
   A=Fresnel(
    ExponentIn=ScalarParameter() {defaultValue=0.748, group=Effect_GemStone_Base, parameterName=EFGM_CoreDarkening}
    BaseReflectFractionIn=ScalarParameter() {defaultValue=0.3, group=Effect_GemStone_Base, parameterName=EFGM_CoreColor}
   ) {baseReflectFraction=0.1}
   B=ScalarParameter() {defaultValue=1.3, group=Effect_GemStone_Base, parameterName=EFGM_CoreMulti}
  )
  NamedRerouteDeclaration(None=Multiply(
    A=MaterialFunctionCall(Value_Morning=@n111, Value_Day=@n112, Value_Evening=@n113, Value_Night=@n114) {materialFunction=&Scalar_Time_of_Day}.Result
    B=Multiply(A=@n116, B=@n117)
   )) {name=EM_Diamond, nodeColor=<rgba 1 0.2344 0 1>}
  n118 = Custom!() {code="return View.DirectionalLightDirection;", description=DirectionalLightVector, inputs []}
  n119 = DotProduct(A=@n118, B=@n118)
  n120 = CollectionParameter() {
   collection=&FortniteMaterialParameters
   parameterName=SunlightDirectionalVector
  }
  n121 = ComponentMask(None=@n120) {b=true, g=true, r=true}
  n122 = MaterialFunctionCall!(In=If(
    A=Custom() {code="return View.DirectionalLightColor;", description=DirectionalLightColor, inputs []}
    "A > B"=If(A=@n119, "A > B"=@n118, "A == B"=@n121, "A < B"=@n121) {desc=check if DirectionalLight exists}
    "A == B"=@n121
    "A < B"=@n121
   ) {desc=check if DirectionalLight color is Black}) {materialFunction=/Game/Packages/Fortress_SharedMaterials/Cinematics/Anime/MF_CINE_AnimeLightDirOverride}
  n123 = MaterialFunctionCall(Float3=@n122.Result) {materialFunction=&BreakOutFloat3Components}
  n124 = ScalarParameter() {
   desc="0.0 = the light vector is locked on the XY plane.\r\n1.0 = the actual light vector is used  \r\n(for Faces use a blend value of 0,0)"
   group=Effect_GemStone_lighting
   parameterName=EFGM_LockedZLightVector_Blend
   sliderMax=1
  }
  n125 = MaterialFunctionCall(X=@n123.R, Y=@n123.R, Z=Multiply(A=@n123.B, B=@n124)) {materialFunction=&MakeFloat3}
  n126 = StaticSwitchParameter(True=LinearInterpolate(A=@n122.Result, B=Normalize(VectorInput=@n125.Result), Alpha=VertexColor().A), False=@n122.Result) {
   desc="Lock the Z axis of the Light Vector\r\nis driven by the VertexColor A channel (0 = Regular LightVector, 1= Locked LIghtVector)"
   group=Effect_GemStone_lighting
   parameterName=EFGM_UseLockedZLightVector
   sortPriority=0
  }
  NamedRerouteDeclaration(None=@n126) {name=LightVector, nodeColor=<rgba 1 0.934 0.5103 1>}
  NamedRerouteDeclaration(None=@n34) {name=RefractColor, nodeColor=<rgba 0 0.3203 1 1>}
  NamedRerouteDeclaration(None=@n32) {name=ReflectColor, nodeColor=<rgba 0 0.3438 1 1>}
  n127 = Transform(None=NamedRerouteUsage())
  n128 = Sine(None=Multiply(A=Fresnel(Normal=@n127) {exponent=1}, B=ScalarParameter() {defaultValue=3, group=Effect_GemStone_Override, parameterName=EFGM_LUTSineMulti}) {constB=3})
  NamedRerouteDeclaration(None=LinearInterpolate(
    A=ScalarParameter() {defaultValue=0.4, group=Effect_GemStone_Override, parameterName=EFGM_LUTMin}
    B=ScalarParameter() {defaultValue=1, group=Effect_GemStone_Override, parameterName=EFGM_LUTMax}
    Alpha=ConstantBiasScale(None=@n128)
   ) {constA=0.4}) {name=FakeLUT, nodeColor=<rgba 1 0.3047 0 1>}
  n129 = NamedRerouteUsage()
  n130 = Sine(None=Multiply(A=Fresnel(Normal=@n129) {exponent=1}) {constB=3})
  NamedRerouteDeclaration(None=LinearInterpolate(Alpha=ConstantBiasScale(None=@n130)) {constA=0.5}) {name=FakeLUT_Interior, nodeColor=<rgba 1 0.3047 0 1>}
  n131 = NamedRerouteUsage()
  n132 = ScalarParameter() {defaultValue=2, group=Effect_GemStone_Base, parameterName=EFGM_InteriorDepth}
  n133 = ScalarParameter() {defaultValue=2, group=Effect_GemStone_Base, parameterName=EFGM_InteriorTiling}
  n134 = ScalarParameter() {defaultValue=1, group=Effect_GemStone_Base, parameterName=EFGM_InteriorBlendContrast}
  n135 = VectorParameter() {group=Effect_GemStone_Base, parameterName=EFGM_InteriorActorOffset, defaultValue=<rgba 0 0 37.5 1>}
  n136 = StaticBool() {value=true}
  NamedRerouteDeclaration(None=Transform(None=MaterialFunctionCall(Texture=@n131, Depth=@n132, Tile=@n133, BlendContrast=@n134, ActorOffset=@n135.RGB, bUseLocalCameraVector=@n136) {materialFunction=&MF_TriplanarVirtualBox}.Result)) {name=InteriorNormal, nodeColor=<rgba 0.1016 1 0 1>}
  NamedRerouteDeclaration(None=MaterialFunctionCall(
    Ray Direction=NamedRerouteUsage()
    Surface Normal=NamedRerouteUsage()
    Refractive Index Origin=NamedRerouteUsage()
    Refractive Index Target=Constant() {r=1}
   ) {materialFunction=&Refract}.Vector) {name=InteriorRefraction, nodeColor=<rgba 1 0 0.6562 1>}
  NamedRerouteDeclaration(None=@n30.Normal) {name=Root_Normal, nodeColor=<rgba 0.4062 0 1 1>}
  NamedRerouteDeclaration(None=@n25.G) {name=Msk_Curvature, nodeColor=<rgba 0 1 0.7422 1>}
  NamedRerouteDeclaration(None=@n24) {name=CM_Mask, nodeColor=<rgba 1 0 0.3047 1>}
  n137 = PixelNormalWS()
  n138 = ScalarParameter() {defaultValue=1, group=Effect_GemStone_Override, parameterName=EFGM_BacklightViewAngle, sortPriority=16}
  n139 = CameraVectorWS()
  n140 = Normalize(VectorInput=Add(A=@n137, B=Multiply(A=@n138, B=@n139)))
  n141 = Multiply(A=SkyAtmosphereLightDirection()) {constB=-1}
  n142 = Add(A=Multiply(A=DotProduct(A=@n140, B=@n141)) {constB=0.5}) {constB=0.5}
  n143 = Multiply(
   A=Power(
    Base=Saturate(None=@n142)
    Exp=ScalarParameter() {defaultValue=2, group=Effect_GemStone_Override, parameterName=EFGM_BacklightAperture, sortPriority=17}
   )
   B=Saturate(None=DotProduct(A=@n137, B=@n139))
  )
  NamedRerouteDeclaration(None=LinearInterpolate(
    A=ScalarParameter() {defaultValue=1, group=Effect_GemStone_Override, parameterName=EFGM_LightScatteringMin}
    B=ScalarParameter() {defaultValue=3, group=Effect_GemStone_Override, parameterName=EFGM_LightScatteringMax}
    Alpha=Saturate(None=@n143)
   ) {constA=1, constB=1.5}) {name=Interior Light Scattering, nodeColor=<rgba 0.5 1 0 1>}
  NamedRerouteUsage()
  ScalarParameter() {defaultValue=3, group=Effect_GemStone_Base, parameterName=EFGM_NormalTile}
  ScalarParameter() {defaultValue=6, group=Effect_Gemstone_RayMarching, parameterName=EFGM_RayMarchingTilling}
  SkyAtmosphereLightDirection()
  StaticSwitchParameter(True=StaticBoolParameter() {group=Effect-Gemstone-Mask, parameterName=EFGM_Decal_Mask-Block}, False=StaticBool()) {group=Customization, parameterName=Function_Left_Right_Decal_V2}
  TextureCoordinate()
  TextureCoordinate()
  TextureSample() {samplerType=SAMPLERTYPE_Normal}
  TransformPosition(None=WorldPosition() {worldPositionShaderOffset=WPT_ExcludeAllShaderOffsets}.XYZ) {transformSourceType=TRANSFORMPOSSOURCE_World}
  Transform(None=NamedRerouteUsage())
  VertexNormalWS()
 ]}
