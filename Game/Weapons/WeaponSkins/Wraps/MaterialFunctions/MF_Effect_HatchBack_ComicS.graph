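# Weapon-wrap material function that applies comic/anime shading with hatching, optionally replacing the input material through the Effect_ComicShading static switch.
# Sunlight-driven NdotL builds shading masks with Voronoi-based hatching, curvature lines, banded metallic, specular and rim light.
# IsState inputs blend tiling, projection and line settings per object type (weapon, vehicle, mech, character, backpack, pickaxe).
# The color-graded result goes through time-of-day adjustment into Emissive, with BaseColor zeroed.
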MaterialFunction /Game/Weapons/WeaponSkins/Wraps/MaterialFunctions/MF_Effect_HatchBack_ComicS
graph {nodes [
  n1 = Power(Base=NamedRerouteUsage()) {constExponent=2.2}
  Add(A=@n1) {constB=-0.3}
  n2 = VectorParameter() {group=Effect-ComicShading--Anime Maps, parameterName=EComicSh_BaseColor_Shaded_Alt_Axies, defaultValue=<rgba 0 0 0 1>}
  n3 = MaterialFunctionCall!(
   Normal_Map=NamedRerouteUsage()
   Offset Rt
   Pivot
   Caculated_Bounds=NamedRerouteUsage()
   Pre_Skinned_Normals=NamedRerouteUsage()
   is_State=NamedRerouteUsage()
   UV_Rotation=NamedRerouteUsage()
   Uvs_Offset
   Uvs_TileRate=NamedRerouteUsage()
   Uvs_Offset_Pivot=NamedRerouteUsage()
   For_Safe_Zones
   Rotation_Comp
   Clamp Y
   Clamp X
   Clamp Padding_Comp
   No_Rotation=StaticBoolParameter() {defaultValue=true, group=Effect-ComicShading--Uvs, parameterName=EComicSh_Rotation_Disabled, sortPriority=0}
  ) {materialFunction=&/MF_Projection_Side_Normal_V2}
  n4 = LinearInterpolate(A=@n2.A, B=TextureSample(UVs=@n3.UVs, Tex=NamedRerouteUsage()).RGBA, Alpha=NamedRerouteUsage())
  ChannelMaskParameter(None=@n4) {group=Effect-ComicShading--Anime Maps, maskChannel=Blue, parameterName=EComicSh_Mask_Channel_Sheen, defaultValue=<rgba 0 0 1 0>}
  ChannelMaskParameter(None=@n4) {group=Effect-ComicShading--Anime Maps, maskChannel=Alpha, parameterName=EComicSh_Mask_Channel_Threshold, defaultValue=<rgba 0 0 0 1>}
  n5 = NamedRerouteUsage()
  n6 = MaterialFunctionCall(Float3=Abs(None=@n5)) {materialFunction=&BreakOutFloat3Components}
  n7 = Constant() {r=0.5}
  n8 = TransformPosition(None=WorldPosition() {worldPositionShaderOffset=WPT_ExcludeAllShaderOffsets}.XYZ) {transformSourceType=TRANSFORMPOSSOURCE_World}
  n9 = ComponentMask(None=@n8) {b=true, desc=YZ, g=true}
  n10 = ComponentMask(None=@n8) {desc=YX, g=true, r=true}
  n11 = ComponentMask(None=@n8) {b=true, desc=XZ, r=true}
  n12 = If(A=@n6.B, B=@n7, "A > B"=@n10, "A < B"=@n11)
  n13 = If(A=@n6.B, "A > B"=@n10, "A < B"=If(A=@n6.R, "A > B"=@n9, "A < B"=@n11))
  n14 = Constant() {r=0.5}
  n15 = If(A=@n6.G, B=@n14, "A > B"=@n11, "A < B"=If(A=@n6.B, B=@n14, "A > B"=@n10, "A < B"=@n9))
  n16 = MaterialFunctionCall(Float4=NamedRerouteUsage()) {materialFunction=&BreakOutFloat4Components}
  n17 = Constant() {r=0.5}
  n18 = LinearInterpolate(
   A=LinearInterpolate(A=If(A=@n6.R, B=@n7, "A > B"=@n9, "A < B"=@n12), B=LinearInterpolate(A=@n13, B=@n15, Alpha=@n16.G), Alpha=@n16.R)
   B=If(A=@n6.G, B=@n17, "A > B"=@n11, "A < B"=If(A=@n6.R, B=@n17, "A > B"=@n9, "A < B"=@n10))
   Alpha=@n16.B
  )
  n19 = ScalarParameter() {defaultValue=0.35, group=Effect-ComicShading--HatchingLines, parameterName=EComicSh_Hatching_Line_Rot, sliderMax=1, sliderMin=-1}
  n20 = Rotator(Coordinate=Add(A=NamedRerouteDeclaration(None=@n18) {name=UVS_Cheap_Tri, nodeColor=<rgba 1 0 0.8438 1>}) {constB=2}, Time=@n19) {speed=3.1415}
  n21 = NamedRerouteUsage()
  n22 = Multiply(A=@n20, B=@n21) {constB=64}
  n23 = Frac(None=@n22)
  n24 = Multiply(A=Abs(None=Add(A=@n23) {constB=-0.5})) {constB=2}
  n25 = OneMinus(None=Frac(None=@n24))
  ComponentMask(None=@n25) {r=true}
  ComponentMask(None=@n25) {g=true}
  n26 = NamedRerouteUsage()
  Composite!(Float3=@n26, False=@n26, InputPin=@n26)
  n27 = LinearInterpolate(
   A=ScalarParameter() {defaultValue=6, group=Effect-ComicShading--HatchingLines, parameterName=EComicSh_Lerp_Max, sliderMax=16}
   B=ScalarParameter() {group=Effect-ComicShading--HatchingLines, parameterName=EComicSh_Lerp_Min, sliderMax=8, sliderMin=-4}
   Alpha=MaterialFunctionCall(
    Input=NamedRerouteUsage()
    Input Low=ScalarParameter() {group=Effect-ComicShading--HatchingLines, parameterName=EComicSh_Hatching_NdotL_Min, sliderMax=1}
    Input High=ScalarParameter() {defaultValue=0.5, group=Effect-ComicShading--HatchingLines, parameterName=EComicSh_Hatching_NdotL_Max, sliderMax=1}
   ) {materialFunction=&RemapValue_0-1}.Result
  ) {constA=6, constB=0}
  n28 = Constant() {r=1}
  Composite!(In=@n27, RampWidth=@n28) {subgraphName=SmoothFloor}
  Composite(In=@n27, RampWidth=@n28) {subgraphName=SmoothRound}
  Constant2Vector()
  Divide(A=@n20, B=@n21)
  n29 = TextureSample() {samplerType=SAMPLERTYPE_Normal, texture=/Game/Weapons/FORT_Rifles/T_SCAR_N}
  n30 = Constant3Vector() {constant=<rgba 1 0 0 0>}
  n31 = GetMaterialAttributes(None=FunctionInput(Preview=SetMaterialAttributes!(MaterialAttributes, Normal=@n29.RGB, BaseColor=@n30) {attributeSetTypes [#1, #2]}) {inputName=Material, inputType=FunctionInput_MaterialAttributes}) {attributeGetTypes [
    #2
    #3
    #1
    #4
    #5
   ]}
  FunctionOutput(None=StaticSwitchParameter(True=NamedRerouteUsage(), False=NamedRerouteDeclaration(None=@n31.MaterialAttributes) {name=Root_Material, nodeColor=<rgba 0.7812 1 0 1>}) {group=Customization, parameterName=Effect_ComicShading})
  n32 = Transform(None=NamedRerouteUsage())
  n33 = VertexColor()
  n34 = Multiply(A=@n33) {constB=2}
  LinearInterpolate(
   A=@n32
   B=Transform(None=Subtract(A=@n34)) {transformSourceType=TRANSFORMSOURCE_Local}
   Alpha=Multiply(A=ScalarParameter() {group=Effect-ComicShading--ComicShading, parameterName=EComicSh_VertexBakedNormal_Blend, sliderMax=1}, B=@n33.A)
  )
  MaterialFunctionCall!(In=@n27, Ramp Width=@n28) {materialFunction=/Engine/Functions/Engine_MaterialFunctions01/Math/SmoothCeil}
  MaterialFunctionCall(Float2=NamedRerouteUsage()) {materialFunction=&BreakOutFloat2Components}
  Multiply(A=TextureSample(UVs=Divide() {constB=125}, Tex=TextureObject() {texture=/Game/Weapons/WeaponSkins/Wraps/Textures/UVCheaker}).RGB)
  Multiply(
   A=@n22
   B=Multiply(A=MaterialFunctionCall() {materialFunction=&ObjectScale}.Scale Z) {constB=-1}
  )
  n35 = MaterialFunctionCall(Float2=NamedRerouteUsage()) {materialFunction=&BreakOutFloat2Components}
  n36 = ScalarParameter() {defaultValue=1, group=Effect-ComicShading--HatchingLines, parameterName=EComicSh_Voronoi_Ratio, sliderMax=2}
  n37 = MaterialFunctionCall(X=@n35.R, Y=Multiply(A=@n35.G, B=Divide(B=@n36))) {materialFunction=&MakeFloat2}
  Multiply(
   A=@n37.Result
   B=ScalarParameter() {defaultValue=128, group=Effect-ComicShading--HatchingLines, parameterName=EComicSh_Voronoi_Tiling, sliderMax=16}
  )
  Multiply(A=@n5)
  n38 = Constant()
  n39 = ScalarParameter() {
   defaultValue=1
   desc=set to 1.0 for 33x33x33 resolution instead 0.0 for 22x22x22.
   group=Effect-ComicShading--LUT
   parameterName=EComicSh_use 33 Resolution
   sliderMax=1
  }
  n40 = Constant() {desc=32 / 33, r=0.9697}
  MaterialExpressionReroute_2 = Reroute(None=@n26)
  n41 = MaterialFunctionCall(Float3=@MaterialExpressionReroute_2.Float3) {materialFunction=&BreakOutFloat3Components}
  MaterialExpressionReroute_3 = Reroute(None=@n26)
  n42 = StaticSwitchParameter(
   True=MaterialFunctionCall(X=@n41.R, Y=@n41.R, Z=@n41.R) {materialFunction=&MakeFloat3}.Result
   False=@MaterialExpressionReroute_3.False
  ) {defaultValue=true, group=Effect-ComicShading--LUT, parameterName=EComicSh_Flip GB, sortPriority=0}
  n43 = Constant() {desc=21 / 22, r=0.9545}
  n44 = If(
   A=@n39
   "A > B"=Add(A=Multiply(A=@n42, B=@n40), B=Constant() {desc=0.5 / 33, r=0.0152}) {desc=33x33x33 resolution}
   "A < B"=Add(A=Multiply(A=@n42, B=@n43), B=Constant() {desc=0.5 / 22, r=0.0227}) {desc=22x22x22 resolution}
  ) {constB=0.5, equalsThreshold=0.001}
  n45 = TextureObjectParameter() {
   group=Effect-ComicShading--LUT
   parameterName=EComicSh_LUT Volume Texture
   sortPriority=1
   texture=/Game/Packages/Fortress_SharedMaterials/Anime_Master_Material/VT_InverseLUT
  }
  n46 = ScalarParameter() {defaultValue=3, group=Effect-ComicShading--LUT, parameterName=EComicSh_LUT Multiplier, sortPriority=5}
  MaterialExpressionReroute_5 = Reroute(None=@n26)
  n47 = ScalarParameter() {defaultValue=0.7, group=Effect-ComicShading--LUT, parameterName=EComicSh_Gain, sliderMax=1.5, sliderMin=0.5, sortPriority=2}
  n48 = Multiply(
   A=Add(
    A=@MaterialExpressionReroute_5.InputPin
    B=ScalarParameter() {group=Effect-ComicShading--LUT, parameterName=EComicSh_Lift, sliderMax=0.2, sortPriority=1}
   )
   B=@n47
  )
  n49 = ScalarParameter() {defaultValue=0.876, group=Effect-ComicShading--LUT, parameterName=EComicSh_Gamma, sliderMax=1.5, sliderMin=0.5, sortPriority=3}
  n50 = LinearInterpolate(
   A=Multiply(
    A=TextureSample(UVs=@n44, Tex=@n45) {texture=&VT_InverseLUT}.RGB
    B=FunctionInput(Preview=@n46) {bUsePreviewValueAsDefault=true, inputName=LUT Multiplier, inputType=FunctionInput_Scalar, sortPriority=300}
   )
   B=MaterialFunctionCall(
    In=Power(Base=@n48, Exp=@n49)
    Power=ScalarParameter() {defaultValue=0.808, group=Effect-ComicShading--LUT, parameterName=EComicSh_SCurve, sliderMax=2, sortPriority=4}
   ) {materialFunction=&SCurve}.Result
   Alpha=ScalarParameter() {defaultValue=1, group=Effect-ComicShading--LUT, parameterName=EComicSh_use Color Adjustment Instead of LUT, sliderMax=1}
  )
  MaterialExpressionReroute_4 = Reroute(None=@n50)
  n51 = LinearInterpolate(
   B=NamedRerouteUsage()
   Alpha=ScalarParameter() {group=Effect-ComicShading--Anime Accent, parameterName="EComicSh_Root_Emissice---->Add", sliderMax=1, sliderMin=-1}
  )
  n52 = SetMaterialAttributes!(
   MaterialAttributes
   BaseColor=@n38
   Specular=@n38
   Opacity=@n38
   SubsurfaceColor=@n38
   Metallic=NamedRerouteUsage()
   EmissiveColor=MaterialFunctionCall!(Color=Add(A=@MaterialExpressionReroute_4.Output, B=@n51)) {materialFunction=/Game/Packages/Fortress_SharedMaterials/Anime_Master_Material/MF_TODColorAdjustment}.Result
   Normal=NamedRerouteUsage()
   AmbientOcclusion
  ) {attributeSetTypes [
    #2
    #6
    #7
    #8
    #4
    #5
    #1
    #9
   ]}
  NamedRerouteDeclaration(None=@n52) {name=Material output, nodeColor=<rgba 0.3594 1 0 1>}
  n53 = ScalarParameter() {defaultValue=128, group=Effect-ComicShading--HatchingLines_Objects, parameterName=EComicSh_Voronoi_Tiling_Vehicle}
  n54 = ScalarParameter() {defaultValue=128, group=Effect-ComicShading--HatchingLines_Objects, parameterName=EComicSh_Voronoi_Tiling_Side}
  n55 = MaterialFunctionCall(Float4=NamedRerouteUsage()) {materialFunction=&BreakOutFloat4Components}
  n56 = LinearInterpolate(
   A=LinearInterpolate(
    A=ScalarParameter() {defaultValue=128, group=Effect-ComicShading--HatchingLines_Objects, parameterName=EComicSh_Voronoi_Tiling_Weapon}
    B=LinearInterpolate(A=@n53, B=@n54, Alpha=@n55.G)
    Alpha=@n55.R
   )
   B=LinearInterpolate(
    A=ScalarParameter() {defaultValue=128, group=Effect-ComicShading--HatchingLines_Objects, parameterName=EComicSh_Voronoi_Tiling_Mech}
    B=ScalarParameter() {defaultValue=128, group=Effect-ComicShading--HatchingLines_Objects, parameterName=EComicSh_Voronoi_Tiling_Character}
    Alpha=@n55.A
   )
   Alpha=@n55.B
  )
  n57 = ScalarParameter() {defaultValue=128, group=Effect-ComicShading--HatchingLines_Objects, parameterName=EComicSh_Voronoi_Tiling_Backpack}
  n58 = MaterialFunctionCall(Float2=NamedRerouteUsage()) {materialFunction=&BreakOutFloat2Components}
  NamedRerouteDeclaration(None=LinearInterpolate(
    A=LinearInterpolate(A=@n56, B=@n57, Alpha=@n58.R)
    B=ScalarParameter() {defaultValue=128, group=Effect-ComicShading--HatchingLines_Objects, parameterName=EComicSh_Voronoi_Tiling_Pickaxe}
    Alpha=@n58.G
   )) {name=Voronoi_Tiling, nodeColor=<rgba 0 0.7188 1 1>}
  NamedRerouteDeclaration(None=@n31.EmissiveColor) {name=Root_Emissive, nodeColor=<rgba 0.4062 0 1 1>}
  NamedRerouteDeclaration(None=@n31.BaseColor) {name=Root_BaseColor, nodeColor=<rgba 0.4062 0 1 1>}
  NamedRerouteDeclaration(None=StaticBoolParameter() {group=Effect-ComicShading--Anime Maps, parameterName="EComicSh_Root_MMaterial------>Two Textures", sortPriority=-1}) {name=Material_State, nodeColor=<rgba 1 0 0.5156 1>}
  n59 = ScalarParameter() {group=Effect-ComicShading--Uvs, parameterName="EComicSh_Rotation_Weapon\\Pickaxe", sortPriority=150}
  n60 = MaterialFunctionCall(Float4=NamedRerouteUsage()) {materialFunction=&BreakOutFloat4Components}
  n61 = LinearInterpolate(
   A=ScalarParameter() {group=Effect-ComicShading--Uvs, parameterName=EComicSh_Rotation_Vehicle, sortPriority=150}
   B=ScalarParameter() {group=Effect-ComicShading--Uvs, parameterName=EComicSh_Rotation_Vehicle_Side, sortPriority=150}
   Alpha=@n60.G
  )
  NamedRerouteDeclaration(None=LinearInterpolate(
    A=LinearInterpolate(A=@n59, B=@n61, Alpha=@n60.R)
    B=ScalarParameter() {group=Effect-ComicShading--Uvs, parameterName="EComicSh_Rotation_Mech\\Character\\Back", sortPriority=150}
    Alpha=@n60.B
   )) {name=UVs_Rotation, nodeColor=<rgba 0 0.7188 1 1>}
  n62 = FunctionInput(Preview=TextureSample() {samplerType=SAMPLERTYPE_Masks, texture=/Game/Weapons/FORT_Rifles/Textures/T_Scar_CM}.RGBA) {inputName=CM_Msk, inputType=FunctionInput_Vector4, sortPriority=4}
  n63 = MaterialFunctionCall(Float4=@n62) {materialFunction=&BreakOutFloat4Components}
  n64 = Add(A=NamedRerouteDeclaration(None=@n63.G) {name=Msk_Curvature, nodeColor=<rgba 0 1 0.7422 1>}) {constB=-0.5}
  NamedRerouteDeclaration(None=Saturate(None=Multiply(A=@n64) {constB=-1})) {name=Curve_Pits, nodeColor=<rgba 1 0 0.8906 1>}
  NamedRerouteDeclaration(None=Saturate(None=@n64)) {name=Curve_Peaks, nodeColor=<rgba 1 0 0.8906 1>}
  n65 = Custom() {code="return View.DirectionalLightDirection;", description=DirectionalLightDirection, inputs []}
  n66 = CollectionParameter() {
   collection=&FortniteMaterialParameters
   parameterName=SunlightDirectionalVector
  }
  n67 = ComponentMask(None=@n66) {b=true, g=true, r=true}
  n68 = If(
   A=ComponentMask(None=Custom() {code="return View.DirectionalLightColor;", description=DirectionalLightColor, outputType=CMOT_Float4, inputs []}) {b=true, g=true, r=true}
   "A > B"=If(A=DotProduct(A=@n65, B=@n65), "A > B"=@n65, "A == B"=@n67, "A < B"=@n67) {desc=check if DirectionalLight exists}
   "A == B"=@n67
   "A < B"=@n67
  ) {desc=check if DirectionalLight color is Black}
  NamedRerouteDeclaration(None=@n68) {name=LightVector, nodeColor=<rgba 0.4766 1 0 1>}
  n69 = NamedRerouteUsage()
  n70 = NamedRerouteUsage()
  n71 = NamedRerouteDeclaration(None=DotProduct(A=@n69, B=@n70)) {name=NdotL, nodeColor=<rgba 0.7344 1 0 1>}
  NamedRerouteDeclaration(None=Add(A=Multiply(A=@n71) {constB=0.5}) {constB=0.5}) {name=NdotL_Lambert, nodeColor=<rgba 0.4531 1 0 1>}
  n72 = Max(A=NamedRerouteUsage(), B=ScalarParameter() {group=Effect-ComicShading--Anime Accent, parameterName="EComicSh_AccentColor   0=active"})
  NamedRerouteDeclaration(None=@n72) {name=Mask_Prime_State, nodeColor=<rgba 0.3359 1 0 1>}
  n73 = VectorParameter() {group=Effect-ComicShading--ComicShading, parameterName=EComicSh_HatchingColor, sortPriority=3, defaultValue=<rgba 0.004 0.004 0.004 1>}
  n74 = NamedRerouteUsage()
  n75 = VectorParameter() {group=Effect-ComicShading--Anime Accent, parameterName=EComicSh_NonPrimay_Shaded, defaultValue=<rgba 0.5 0.5 0.5 1>}
  n76 = VectorParameter() {group=Effect-ComicShading--Anime Accent, parameterName=EComicSh_NonPrimay_Lit, defaultValue=<rgba 1 1 1 1>}
  n77 = NamedRerouteUsage()
  n78 = LinearInterpolate(A=Multiply(A=@n74, B=@n75.RGB), B=Multiply(A=@n76.RGB, B=@n74), Alpha=@n77)
  n79 = VectorParameter() {group=Effect-ComicShading--Anime Accent, parameterName=EComicSh_AccentColor_lit, defaultValue=<rgba 1 0.0781 0.0781 1>}
  n80 = VectorParameter() {group=Effect-ComicShading--Anime Accent, parameterName=EComicSh_AccentColor_Shaded, defaultValue=<rgba 0.5 0.0365 0.0365 1>}
  n81 = LinearInterpolate(
   A=LinearInterpolate(A=@n79.RGB, B=@n80.RGB, Alpha=@n77)
   B=LinearInterpolate(A=NamedRerouteUsage(), B=NamedRerouteUsage(), Alpha=@n77)
   Alpha=@n72
  )
  n82 = Max(A=NamedRerouteUsage(), B=ScalarParameter() {group=Effect-ComicShading--Anime Accent, parameterName="EComicSh_Non Primary   0=active"})
  n83 = VectorParameter() {group=Effect-ComicShading--Anime Maps, parameterName=EComicSh_BaseColor_Tint_Shaded, defaultValue=<rgba 0.5 0.5 0.5 1>}
  n84 = NamedRerouteUsage()
  n85 = Multiply(A=@n83.RGB, B=@n84)
  n86 = VectorParameter() {group=Effect-ComicShading--Anime Maps, parameterName=EComicSh_BaseColor_Tint_Lit, defaultValue=<rgba 1 1 1 1>}
  n87 = Multiply(A=@n84, B=@n86.RGB)
  n88 = NamedRerouteUsage()
  n89 = Saturate(None=StaticSwitch(True=LinearInterpolate(A=@n78, B=@n81, Alpha=@n82), False=LinearInterpolate(A=@n85, B=@n87, Alpha=@n88), Value=NamedRerouteUsage()))
  n90 = Multiply(A=@n89, B=NamedRerouteUsage())
  n91 = NamedRerouteUsage()
  n92 = VectorParameter() {group=Effect-ComicShading--ComicShading, parameterName=EComicSh_RimLight_Color, defaultValue=<rgba 0.3 0.5 1 2>}
  n93 = LinearInterpolate(
   A=Add(A=LinearInterpolate(A=@n89, B=@n90, Alpha=@n91) {desc=metallic}, B=NamedRerouteUsage())
   B=Multiply(A=@n92.RGB, B=@n92.RGB)
   Alpha=NamedRerouteUsage()
  )
  n94 = Multiply(
   A=NamedRerouteUsage()
   B=Max(
    A=NamedRerouteUsage()
    B=ScalarParameter() {group=Effect-ComicShading--DistanceFieldLines_Objects, parameterName="EComicSh_Curavture_Line_Opacity    1=Off"}
   )
  )
  NamedRerouteDeclaration(None=Saturate(None=LinearInterpolate(A=@n73.RGB, B=@n93, Alpha=@n94))) {name=BaseColor_Raw, nodeColor=<rgba 1 0.4506 0.2387 1>}
  n95 = NamedRerouteUsage()
  n96 = ScalarParameter() {defaultValue=1, group=Effect-ComicShading--Shadding, parameterName=EComicSh_Threshold_Exponent}
  NamedRerouteDeclaration(None=LinearInterpolate(
    A=ScalarParameter() {group=Effect-ComicShading--Shadding, parameterName=EComicSh_Threshold_Min, sliderMax=1}
    B=ScalarParameter() {defaultValue=1, group=Effect-ComicShading--Shadding, parameterName=EComicSh_Threshold_Max, sliderMax=1}
    Alpha=Power(Base=@n95, Exp=@n96)
   )) {name=Threshold, nodeColor=<rgba 0 0.5 1 1>}
  n97 = Multiply(
   A=NamedRerouteUsage()
   B=ScalarParameter() {defaultValue=0.5, group=Effect-ComicShading--HatchingLines, parameterName=EComicSh_LineDistortion_RandScale, sliderMax=1}
  )
  n98 = NamedRerouteUsage()
  n99 = ScalarParameter() {defaultValue=0.5, group=Effect-ComicShading--HatchingLines, parameterName=EComicSh_LineDistortion_RandRot, sliderMax=0.25}
  n100 = Rotator(
   Coordinate=Multiply(A=NamedRerouteUsage(), B=Add(A=@n97))
   Time=Add(
    A=ScalarParameter() {group=Effect-ComicShading--HatchingLines, parameterName=EComicSh_Lines_RotationOffset, sliderMax=1}
    B=Multiply(A=@n98, B=@n99)
   )
  ) {centerX=0, centerY=0, speed=-6.283}
  n101 = Frac(None=Multiply(A=ComponentMask(None=@n100) {r=true}, B=NamedRerouteUsage()) {constB=32})
  n102 = Abs(None=Add(A=Multiply(A=@n101) {constB=2}) {constB=-1}) {desc=HatchingDF}
  n103 = NamedRerouteUsage()
  NamedRerouteDeclaration(None=Saturate(None=LinearInterpolate(B=@n102, Alpha=@n103) {constA=1})) {name=UVHatching, nodeColor=<rgba 0.9922 1 0 1>}
  n104 = NamedRerouteUsage()
  n105 = DotProduct(A=ReflectionVectorWS(CustomWorldNormal=NamedRerouteUsage()), B=NamedRerouteUsage()) {desc=Phong}
  n106 = VectorParameter() {group=Effect-ComicShading--ComicShading, parameterName="EComicSh_Specular_Color\\Intensity", defaultValue=<rgba 0.5 0.5 0.5 1>}
  n107 = Multiply(
   A=SphereMask(
    A=Saturate(None=@n105)
    B=Constant() {r=1}
    Radius=ScalarParameter() {defaultValue=0.1, group=Effect-ComicShading--ComicShading, parameterName=EComicSh_Specular_Size, sliderMax=1}
    Hardness=ScalarParameter() {defaultValue=0.8, group=Effect-ComicShading--ComicShading, parameterName=EComicSh_Specular_Hardness, sliderMax=1}
   ) {hardnessPercent=78}
   B=@n106.RGB
  )
  NamedRerouteDeclaration(None=Saturate(None=Multiply(A=@n104, B=@n107))) {name=Specular_Curve, nodeColor=<rgba 1 0.4453 0 1>}
  n108 = NamedRerouteUsage()
  n109 = VectorParameter() {group=Effect-ComicShading--Anime Maps, parameterName=EComicSh_BaseColor_Lit_Alt_Axies, defaultValue=<rgba 1 1 1 1>}
  n110 = NamedRerouteUsage()
  n111 = LinearInterpolate(
   A=ScalarParameter() {defaultValue=0.3, group=Effect-ComicShading--Anime Accent, parameterName="EComicSh_Accent_Sheen\\Roughness"}
   B=Multiply(
    A=@n109.A
    B=ScalarParameter() {defaultValue=0.3, group=Effect-ComicShading--Anime Maps, parameterName=EComicSh_BaseColor_Lit_Alphas_Sheen_Mulptyer}
   )
   Alpha=@n110
  )
  n112 = NamedRerouteUsage()
  NamedRerouteDeclaration(None=StaticSwitch(True=LinearInterpolate(A=@n108, B=@n111, Alpha=@n112), False=@n108, Value=NamedRerouteUsage())) {name=Sheen, nodeColor=<rgba 0 0.5 1 1>}
  n113 = Constant() {r=0.5}
  n114 = ScalarParameter() {defaultValue=0.05, group=Effect-ComicShading--ComicShading, parameterName=EComicSh_ShadedBaseColor_TransitionAmount, sliderMax=1}
  n115 = Multiply(A=Saturate(None=@n114)) {constB=0.5}
  n116 = NamedRerouteUsage()
  n117 = ScalarParameter() {defaultValue=-0.3, group=Effect-ComicShading--ComicShading, parameterName=EComicSh_ShadedBaseColor_Bias, sliderMax=1, sliderMin=1}
  n118 = Add(A=Multiply(A=Subtract(A=@n116, B=@n117)) {constB=0.5}) {constB=0.5}
  n119 = LinearInterpolate(
   A=@n118
   B=MaterialFunctionCall(Base=NamedRerouteUsage(), Blend=@n118) {materialFunction=&Blend_Overlay}.Result
   Alpha=ScalarParameter() {group=Effect-ComicShading--ComicShading, parameterName=EComicSh_ShadedBaseColor_ThresholdBlend, sliderMax=1}
  )
  n120 = NamedRerouteUsage()
  NamedRerouteDeclaration(None=SmoothStep(
    Min=Subtract(A=@n113, B=@n115) {constA=0.5}
    Max=Add(A=@n113, B=@n115) {constA=0.5, constB=-0.5}
    Value=LinearInterpolate(A=@n119, Alpha=@n120)
   ) {constMax=0.01}) {name=ShadedBaseColor_Curve, nodeColor=<rgba 0 0.9439 1 1>}
  n121 = Add(A=@n69, B=CameraVectorWS())
  NamedRerouteDeclaration(None=DotProduct(A=Normalize(VectorInput=@n121), B=@n70)) {name=NdotH, nodeColor=<rgba 0.7344 1 0 1>}
  n122 = NamedRerouteUsage()
  n123 = NamedRerouteUsage()
  NamedRerouteDeclaration(None=Multiply(A=Power(Base=@n122, Exp=@n123) {constExponent=0.5}, B=NamedRerouteUsage()) {constB=3}) {name=DF, nodeColor=<rgba 0 1 0.6166 1>}
  n124 = ScalarParameter() {defaultValue=0.4, group=Effect-ComicShading--DistanceFieldLines, parameterName=EComicSh_DF_LineThickness_Near, sliderMax=1, sortPriority=5}
  n125 = ScalarParameter() {defaultValue=0.6, group=Effect-ComicShading--DistanceFieldLines, parameterName=EComicSh_DF_LineThickness_Far, sliderMax=1, sortPriority=6}
  n126 = Divide(
   A=Subtract(
    A=PixelDepth().R
    B=ScalarParameter() {defaultValue=50, group=Effect-ComicShading--DistanceFieldLines, parameterName=EComicSh_DF_FadeStart, sliderMax=200}
   )
   B=ScalarParameter() {defaultValue=500, group=Effect-ComicShading--DistanceFieldLines, parameterName=EComicSh_DF_FadeDistance, sliderMax=1000}
  )
  n127 = ScalarParameter() {defaultValue=0.9, group=Effect-ComicShading--DistanceFieldLines, parameterName=EComicSh_DF_LineHardness, sliderMax=1, sortPriority=7}
  n128 = SphereMask(A=NamedRerouteUsage(), B=Constant() {r=1}, Radius=LinearInterpolate(A=@n124, B=@n125, Alpha=Saturate(None=@n126)), Hardness=@n127) {attenuationRadius=0.1}
  NamedRerouteDeclaration(None=OneMinus(None=Saturate(None=@n128))) {name=DF_Mask, nodeColor=<rgba 1 0.6562 0 1>}
  n129 = ScalarParameter() {defaultValue=0.5, group=Effect-ComicShading--DistanceFieldLines_Objects, parameterName=EComicSh_DF_Curvature_FallOff_Vehicle}
  n130 = ScalarParameter() {defaultValue=0.5, group=Effect-ComicShading--DistanceFieldLines_Objects, parameterName=EComicSh_DF_Curvature_FallOff_Vehicle_Side}
  n131 = MaterialFunctionCall(Float4=NamedRerouteUsage()) {materialFunction=&BreakOutFloat4Components}
  n132 = LinearInterpolate(
   A=LinearInterpolate(
    A=ScalarParameter() {defaultValue=0.5, group=Effect-ComicShading--DistanceFieldLines_Objects, parameterName=EComicSh_DF_Curvature_FallOff_Weapon}
    B=LinearInterpolate(A=@n129, B=@n130, Alpha=@n131.G)
    Alpha=@n131.R
   )
   B=LinearInterpolate(
    A=ScalarParameter() {defaultValue=0.5, group=Effect-ComicShading--DistanceFieldLines_Objects, parameterName=EComicSh_DF_Curvature_FallOff_Mech}
    B=ScalarParameter() {defaultValue=0.5, group=Effect-ComicShading--DistanceFieldLines_Objects, parameterName=EComicSh_DF_Curvature_FallOff_Character}
    Alpha=@n131.A
   )
   Alpha=@n131.B
  )
  n133 = ScalarParameter() {defaultValue=0.5, group=Effect-ComicShading--DistanceFieldLines_Objects, parameterName=EComicSh_DF_Curvature_FallOff_Backpack}
  n134 = MaterialFunctionCall(Float2=NamedRerouteUsage()) {materialFunction=&BreakOutFloat2Components}
  NamedRerouteDeclaration(None=LinearInterpolate(
    A=LinearInterpolate(A=@n132, B=@n133, Alpha=@n134.R)
    B=ScalarParameter() {defaultValue=0.5, group=Effect-ComicShading--DistanceFieldLines_Objects, parameterName=EComicSh_DF_Curvature_FallOff_Pickaxe}
    Alpha=@n134.G
   )) {name=Curvature_FallOff, nodeColor=<rgba 0 0.7188 1 1>}
  n135 = Constant() {r=0.5}
  n136 = Multiply(A=Saturate(None=ScalarParameter() {defaultValue=0.1, group=Effect-ComicShading--ComicShading, parameterName=EComicSh_LineTransition_Amount, sliderMax=1})) {constB=0.5}
  n137 = Subtract(A=@n135, B=@n136) {constA=0.5}
  n138 = Add(A=@n135, B=@n136) {constA=0.5, constB=-0.5}
  n139 = ScalarParameter() {defaultValue=-0.5, group=Effect-ComicShading--ComicShading, parameterName=EComicSh_ShadingCurve_Bias, sliderMax=1, sliderMin=-1}
  n140 = Multiply(A=Subtract(A=NamedRerouteUsage(), B=@n139)) {constB=0.5}
  n141 = NamedRerouteUsage()
  n142 = ScalarParameter() {defaultValue=1, group=Effect-ComicShading--ComicShading, parameterName=EComicSh_ShadingCurve_CameraExponent, sliderMax=32}
  n143 = Multiply(
   A=LinearInterpolate(
    A=Add(A=@n140) {constB=0.5}
    B=Power(Base=@n141, Exp=@n142) {constExponent=4}
    Alpha=ScalarParameter() {group=Effect-ComicShading--ComicShading, parameterName=EComicSh_ShadingCurve_CameraBlend, sliderMax=1}
   )
   B=Constant() {r=2}
  )
  n144 = NamedRerouteUsage()
  n145 = NamedRerouteUsage()
  n146 = Subtract(
   A=Multiply(
    A=StaticSwitchParameter(True=@n144, False=@n145) {defaultValue=true, group=Effect-ComicShading--HatchingLines, parameterName=EComicSh_UseCrossHatching, sortPriority=0}
    B=ScalarParameter() {defaultValue=0.5, group=Effect-ComicShading--ComicShading, parameterName=EComicSh_LineVariation_Intensity, sliderMax=1}
   ) {constB=0.2}
   B=Constant() {r=1}
  )
  n147 = MaterialFunctionCall(Base=NamedRerouteUsage(), Blend=Saturate(None=Add(A=@n143, B=@n146)) {desc=height blend with 1 heighmap}) {materialFunction=&Blend_Overlay}
  NamedRerouteDeclaration(None=LinearInterpolate(A=SmoothStep(Min=@n137, Max=@n138, Value=@n147.Result) {constMax=0.01}, Alpha=NamedRerouteUsage())) {name=ShadingMask, nodeColor=<rgba 1 0.6562 0 1>}
  n148 = NamedRerouteUsage()
  n149 = ScalarParameter() {defaultValue=5, group=Effect-ComicShading--ComicShading, parameterName=EComicSh_BackLit_Exponent, sliderMax=8}
  n150 = Saturate(None=SmoothStep(Value=Power(Base=@n148, Exp=@n149) {constExponent=4}) {constMax=0, constMin=0.02})
  n151 = MaterialFunctionCall() {materialFunction=&CameraDirectionVector}
  n152 = NamedRerouteUsage()
  n153 = Clamp(
   None=Add(
    A=DotProduct(A=@n151.Result, B=@n152)
    B=ScalarParameter() {defaultValue=0.25, group=Effect-ComicShading--ComicShading, parameterName=EComicSh_BackLit_Bias, sliderMax=1}
   )
   Max=ScalarParameter() {defaultValue=0.75, group=Effect-ComicShading--ComicShading, parameterName=EComicSh_BackLit_MaxBias, sliderMax=1}
  )
  NamedRerouteDeclaration(None=Multiply(
    A=NamedRerouteDeclaration(None=@n150) {name=RimLight_Mask, nodeColor=<rgba 0 1 0.5078 1>}
    B=NamedRerouteDeclaration(None=@n153) {name=BackLit_Curve, nodeColor=<rgba 0.9507 0.45 1 1>}
   )) {name=BackLitShading_Curve, nodeColor=<rgba 0.1307 1 0.334 1>}
  n154 = TextureSampleParameter2D(UVs=Divide(A=@n37.Result, B=NamedRerouteUsage())) {
   group=Effect-ComicShading--Anime Maps
   parameterName=VoronoiMap
   samplerType=SAMPLERTYPE_LinearColor
   sortPriority=1
   texture=/Game/Global/Materials/Character_Parent_2023/Anime_Parent_2023/Textures/T_Hatching_Voronoi
  }
  NamedRerouteDeclaration(None=@n154.G) {name=VoronoiID, nodeColor=<rgba 1 0.2603 0.2703 1>}
  NamedRerouteDeclaration(None=FunctionInput(Preview=Constant4Vector() {constant=<rgba 0 0 1 0>}.RGBA) {bUsePreviewValueAsDefault=true, inputName=IsState, inputType=FunctionInput_Vector4, sortPriority=4}) {name=IsState, nodeColor=<rgba 0.1497 1 0.1633 1>}
  NamedRerouteDeclaration(None=FunctionInput(Preview=Constant4Vector().RGBA) {bUsePreviewValueAsDefault=true, inputName=IsState_B, inputType=FunctionInput_Vector4, sortPriority=4}) {name=IsState_B, nodeColor=<rgba 0.1497 1 0.1633 1>}
  n155 = PreSkinnedLocalBounds()
  n156 = Subtract(A=VertexInterpolator(VS=PreSkinnedPosition()).PS, B=@n155.Min)
  n157 = MaterialFunctionCall(Color=Subtract(A=@n155.Max, B=@n155.Max)) {materialFunction=&SplitComponents}
  n158 = Max(A=Max(A=@n157.R, B=@n157.R), B=@n157.B)
  NamedRerouteDeclaration(None=FunctionInput(Preview=Divide(A=@n156, B=@n158)) {bUsePreviewValueAsDefault=true, inputName=Calcated_Bounds, sortPriority=11}) {name=Caclcated_Bounds, nodeColor=<rgba 0 0.9297 1 1>}
  n159 = PreSkinnedNormal()
  NamedRerouteDeclaration(None=FunctionInput(Preview=VertexInterpolator(VS=@n159).PS) {bUsePreviewValueAsDefault=true, inputName=Pre-Skinned Local Normal, sortPriority=10}) {name=PreSkin_Normal, nodeColor=<rgba 0.8281 1 0 1>}
  NamedRerouteDeclaration(None=@n82) {name=Mask_non_Prim_State, nodeColor=<rgba 0.3359 1 0 1>}
  NamedRerouteDeclaration(None=SmoothStep(Value=@n154.R) {constMax=0.8, constMin=1}) {name=VoronoiIEdgeDistance, nodeColor=<rgba 1 0.2603 0.2703 1>}
  NamedRerouteDeclaration(None=@n154.B) {name=VoronoiICenterDistance, nodeColor=<rgba 1 0.2603 0.2703 1>}
  n160 = Saturate(None=@n1)
  n161 = Constant() {desc=Metallic Band Amount, r=3}
  n162 = Saturate(None=ScalarParameter() {defaultValue=0.1, group=Effect-ComicShading--ComicShading, parameterName=EComicSh_Metallic_Roughness, sliderMax=1})
  n163 = Divide(
   A=MaterialFunctionCall(In=Multiply(A=@n160, B=@n161), Ramp Width=Multiply(A=@n162) {constB=2}) {materialFunction=&SmoothCeil}.Result
   B=@n161
  )
  NamedRerouteDeclaration(None=Power(Base=Saturate(None=@n163)) {constExponent=2.2, desc=gamma correction}) {name=Metallic_Gradient, nodeColor=<rgba 0.1107 1 0.7992 1>}
  n164 = Saturate(None=DotProduct(A=NamedRerouteUsage(), B=PixelNormalWS()))
  n165 = ScalarParameter() {defaultValue=5, group=Effect-ComicShading--ComicShading, parameterName=EComicSh_RimLight_FresnelExponent, sliderMax=12}
  n166 = PixelNormalWS()
  n167 = Saturate(None=InverseLinearInterpolate(
    A=ScalarParameter() {defaultValue=0.1, group=Effect-ComicShading--ComicShading, parameterName=EComicSh_RimLight_Min, sliderMax=1}
    B=ScalarParameter() {defaultValue=0.2, group=Effect-ComicShading--ComicShading, parameterName=EComicSh_RimLight_Max, sliderMax=1}
    Value=Fresnel(ExponentIn=@n165, Normal=@n166) {baseReflectFraction=0}
   ) {constA=0.1, constB=0.2})
  NamedRerouteDeclaration(None=Saturate(None=Multiply(A=@n164, B=@n167))) {name=RimLight, nodeColor=<rgba 1 0 0.1172 1>}
  n168 = NamedRerouteUsage()
  n169 = ChannelMaskParameter(None=@n4) {group=Effect-ComicShading--Anime Maps, maskChannel=Green, parameterName=EComicSh_Mask_Channel_MetalMask, defaultValue=<rgba 0 1 0 0>}
  n170 = LinearInterpolate(
   A=ScalarParameter() {group=Effect-ComicShading--Anime Accent, parameterName=EComicSh_Acent_Metalic}
   B=StaticSwitchParameter(True=@n169.R, False=ScalarParameter() {group=Effect-ComicShading--Anime Maps, parameterName=EComicSh_BaseColor_Metalic}) {group=Effect-ComicShading--Anime Maps, parameterName=EComicSh_Metalic---Mask_Tex}
   Alpha=@n110
  )
  NamedRerouteDeclaration(None=StaticSwitch(True=LinearInterpolate(A=@n168, B=@n170, Alpha=@n112), False=@n168, Value=NamedRerouteUsage())) {name=MetallicMask, nodeColor=<rgba 0 1 0.4844 1>}
  n171 = OneMinus(None=Saturate(None=@n24))
  n172 = Floor(None=Add(A=@n22, B=Constant() {r=256}))
  n173 = Constant() {r=4}
  n174 = OneMinus(None=Saturate(None=Fmod(A=@n172, B=@n173))) {desc=1}
  n175 = MaterialFunctionCall(Float2=@n174) {materialFunction=&BreakOutFloat2Components}
  n176 = Constant() {r=2}
  n177 = OneMinus(None=Saturate(None=Fmod(A=@n172, B=@n176))) {desc=1}
  n178 = MaterialFunctionCall(Float2=Subtract(A=@n177, B=@n174)) {materialFunction=&BreakOutFloat2Components}
  n179 = MaterialFunctionCall(Float2=Subtract(A=Constant() {r=1}, B=@n177)) {materialFunction=&BreakOutFloat2Components}
  n180 = MaterialFunctionCall(Float3=Multiply(
    A=ComponentMask(None=@n171) {r=true}
    B=MaterialFunctionCall(X=@n175.R, Y=@n178.R, Z=@n179.R) {materialFunction=&MakeFloat3}.Result
   )) {materialFunction=&BreakOutFloat3Components}
  MaterialExpressionReroute_8 = Reroute(None=@n27)
  n181 = Frac(None=Add(A=@MaterialExpressionReroute_8.In) {constB=-0.5})
  MaterialExpressionReroute_12 = Reroute(None=@n28)
  n182 = OneMinus(None=@MaterialExpressionReroute_12.RampWidth)
  n183 = Max(A=Divide(A=Subtract(A=@n181, B=@n182), B=@MaterialExpressionReroute_12.RampWidth), B=Constant() {r=0.5}) {desc=It is good to make functions look like a lancer}
  n184 = Add(A=Round(None=@MaterialExpressionReroute_8.In), B=ConstantBiasScale(None=Cosine(None=@n183)))
  MaterialExpressionReroute_14 = Reroute(None=@n184)
  n185 = Add(A=@MaterialExpressionReroute_14.Output) {constB=-1}
  n186 = Max(A=Multiply(A=@n180.R, B=Saturate(None=@MaterialExpressionReroute_14.Output)), B=Multiply(A=@n180.G, B=Saturate(None=@n185)))
  n187 = Add(A=@n185) {constB=-1}
  n188 = Multiply(A=@n180.B, B=Saturate(None=@n187))
  n189 = MaterialFunctionCall(Float3=Multiply(
    A=ComponentMask(None=@n171) {g=true}
    B=MaterialFunctionCall(X=@n175.G, Y=@n178.G, Z=@n179.G) {materialFunction=&MakeFloat3}.Result
   )) {materialFunction=&BreakOutFloat3Components}
  n190 = Add(A=@n187) {constB=-1}
  n191 = Saturate(None=@n190)
  n192 = Add(A=@n190) {constB=-1}
  n193 = Max(A=Max(A=Max(A=@n186, B=@n188), B=Multiply(A=@n189.R, B=@n191)), B=Multiply(A=@n189.G, B=Saturate(None=@n192)))
  n194 = Multiply(A=@n189.B, B=Saturate(None=Add(A=@n192) {constB=-1}))
  NamedRerouteDeclaration(None=OneMinus(None=Max(A=@n193, B=@n194))) {name=LinearHatching, nodeColor=<rgba 1 0 0.0938 1>}
  n195 = ScalarParameter() {defaultValue=3, group=Effect-ComicShading--DistanceFieldLines_Objects, parameterName=EComicSh_DF_Curvature_Strength_Vehicle}
  n196 = ScalarParameter() {defaultValue=3, group=Effect-ComicShading--DistanceFieldLines_Objects, parameterName=EComicSh_DF_Curvature_Strength_Vehicle_Side}
  n197 = MaterialFunctionCall(Float4=NamedRerouteUsage()) {materialFunction=&BreakOutFloat4Components}
  n198 = LinearInterpolate(
   A=LinearInterpolate(
    A=ScalarParameter() {defaultValue=3, group=Effect-ComicShading--DistanceFieldLines_Objects, parameterName=EComicSh_DF_Curvature_Strength_Weapon}
    B=LinearInterpolate(A=@n195, B=@n196, Alpha=@n197.G)
    Alpha=@n197.R
   )
   B=LinearInterpolate(
    A=ScalarParameter() {defaultValue=3, group=Effect-ComicShading--DistanceFieldLines_Objects, parameterName=EComicSh_DF_Curvature_Strength_Mech}
    B=ScalarParameter() {defaultValue=3, group=Effect-ComicShading--DistanceFieldLines_Objects, parameterName=EComicSh_DF_Curvature_Strength_Character}
    Alpha=@n197.A
   )
   Alpha=@n197.B
  )
  n199 = ScalarParameter() {defaultValue=3, group=Effect-ComicShading--DistanceFieldLines_Objects, parameterName=EComicSh_DF_Curvature_Strength_Backpack}
  n200 = MaterialFunctionCall(Float2=NamedRerouteUsage()) {materialFunction=&BreakOutFloat2Components}
  NamedRerouteDeclaration(None=LinearInterpolate(
    A=LinearInterpolate(A=@n198, B=@n199, Alpha=@n200.R)
    B=ScalarParameter() {defaultValue=3, group=Effect-ComicShading--DistanceFieldLines_Objects, parameterName=EComicSh_DF_Curvature_Strength_Pickaxe}
    Alpha=@n200.G
   )) {name=Curvature_Strength_, nodeColor=<rgba 0 0.7188 1 1>}
  n201 = ScalarParameter() {defaultValue=1.5, group=Effect-ComicShading--HatchingLines_Objects, parameterName=EComicSh_Hatching_Line_Tiling_Vehicle}
  n202 = ScalarParameter() {defaultValue=1.5, group=Effect-ComicShading--HatchingLines_Objects, parameterName=EComicSh_Hatching_Line_Tiling_Side}
  n203 = MaterialFunctionCall(Float4=NamedRerouteUsage()) {materialFunction=&BreakOutFloat4Components}
  n204 = LinearInterpolate(
   A=LinearInterpolate(
    A=ScalarParameter() {defaultValue=1.5, group=Effect-ComicShading--HatchingLines_Objects, parameterName=EComicSh_Hatching_Line_Tiling_Weapon}
    B=LinearInterpolate(A=@n201, B=@n202, Alpha=@n203.G)
    Alpha=@n203.R
   )
   B=LinearInterpolate(
    A=ScalarParameter() {defaultValue=1.5, group=Effect-ComicShading--HatchingLines_Objects, parameterName=EComicSh_Hatching_Line_Tiling_Mech}
    B=ScalarParameter() {defaultValue=1.5, group=Effect-ComicShading--HatchingLines_Objects, parameterName=EComicSh_Hatching_Line_Tiling_Character}
    Alpha=@n203.A
   )
   Alpha=@n203.B
  )
  n205 = ScalarParameter() {defaultValue=1.5, group=Effect-ComicShading--HatchingLines_Objects, parameterName=EComicSh_Hatching_Line_Tiling_Backpack}
  n206 = MaterialFunctionCall(Float2=NamedRerouteUsage()) {materialFunction=&BreakOutFloat2Components}
  NamedRerouteDeclaration(None=LinearInterpolate(
    A=LinearInterpolate(A=@n204, B=@n205, Alpha=@n206.R)
    B=ScalarParameter() {defaultValue=1.5, group=Effect-ComicShading--HatchingLines_Objects, parameterName=EComicSh_Hatching_Line_Tiling_Pickaxe}
    Alpha=@n206.G
   )) {name=Hatching_Line, nodeColor=<rgba 0 0.7188 1 1>}
  NamedRerouteDeclaration(None=@n31.Normal) {name=Root_Normal, nodeColor=<rgba 0.4062 0 1 1>}
  NamedRerouteDeclaration(None=@n62) {name=CM_Mask, nodeColor=<rgba 1 0 0.3047 1>}
  n207 = VectorParameter() {group=Effect-ComicShading--Uvs, parameterName=EComicSh_Scale_Weapon, defaultValue=<rgba 1 1 1 1>}
  n208 = VectorParameter() {group=Effect-ComicShading--Uvs, parameterName=EComicSh_Scale_Vehicle, defaultValue=<rgba 1 1 1 1>}
  n209 = Multiply(A=@n208.RGB, B=@n208.RGB)
  n210 = VectorParameter() {group=Effect-ComicShading--Uvs, parameterName=EComicSh_Scale_VehicleSide, defaultValue=<rgba 1 1 1 1>}
  n211 = Multiply(A=@n210.RGB, B=@n210.RGB)
  n212 = MaterialFunctionCall(Float4=NamedRerouteUsage()) {materialFunction=&BreakOutFloat4Components}
  n213 = VectorParameter() {group=Effect-ComicShading--Uvs, parameterName=EComicSh_Scale_Mech, defaultValue=<rgba 1 1 1 1>}
  n214 = VectorParameter() {group=Effect-ComicShading--Uvs, parameterName=EComicSh_Scale_Character, defaultValue=<rgba 1 1 1 1>}
  n215 = LinearInterpolate(
   A=LinearInterpolate(A=Multiply(A=@n207.RGB, B=@n207.RGB), B=LinearInterpolate(A=@n209, B=@n211, Alpha=@n212.G), Alpha=@n212.R)
   B=LinearInterpolate(A=Multiply(A=@n213.RGB, B=@n213.RGB), B=Multiply(A=@n214.RGB, B=@n214.RGB), Alpha=@n212.A)
   Alpha=@n212.B
  )
  n216 = VectorParameter() {group=Effect-ComicShading--Uvs, parameterName=EComicSh_Scale_Backpack, defaultValue=<rgba 1 1 1 1>}
  n217 = Multiply(A=@n216.RGB, B=@n216.RGB)
  n218 = MaterialFunctionCall(Float2=NamedRerouteUsage()) {materialFunction=&BreakOutFloat2Components}
  n219 = VectorParameter() {group=Effect-ComicShading--Uvs, parameterName=EComicSh_Scale_Pickaxe, defaultValue=<rgba 1 1 1 1>}
  NamedRerouteDeclaration(None=LinearInterpolate(A=LinearInterpolate(A=@n215, B=@n217, Alpha=@n218.R), B=Multiply(A=@n219.RGB, B=@n219.RGB), Alpha=@n218.G)) {name=UVs_Scale, nodeColor=<rgba 0 0.625 1 1>}
  n220 = NamedRerouteUsage()
  n221 = NamedRerouteUsage()
  NamedRerouteDeclaration(None=LinearInterpolate(A=@n109.RGB, B=TextureSample(UVs=@n3.UVs, Tex=@n220).RGB, Alpha=@n221)) {name=ColorLit, nodeColor=<rgba 0 0.3438 1 1>}
  n222 = VectorParameter() {group=Effect-ComicShading--Anime Maps, parameterName=EComicSh_BaseColor_Shaded_Alt_Axies, defaultValue=<rgba 0 0 0 1>}
  n223 = NamedRerouteUsage()
  NamedRerouteDeclaration(None=LinearInterpolate(A=@n222.RGB, B=TextureSample(UVs=@n3.UVs, Tex=@n223).RGB, Alpha=@n221)) {name=ColorShaded, nodeColor=<rgba 0.6641 1 0 1>}
  n224 = NamedRerouteDeclaration(None=DotProduct(A=CameraVectorWS(), B=@n70)) {name=NdotV, nodeColor=<rgba 0.7344 1 0 1>}
  NamedRerouteDeclaration(None=Add(A=Multiply(A=@n224) {constB=0.5}) {constB=0.5}) {name=NdotV_Lambert, nodeColor=<rgba 0.7344 1 0 1>}
  NamedRerouteDeclaration(None=Normalize(VectorInput=@n32)) {name=WSNormal, nodeColor=<rgba 0 0.4375 1 1>}
  n225 = NamedRerouteUsage()
  n226 = LinearInterpolate(A=Constant3Vector(), B=@n3.Raw_Normal, Alpha=NamedRerouteUsage())
  n227 = LinearInterpolate(
   A=@n225
   B=LinearInterpolate(
    A=@n225
    B=Add(A=@n225, B=@n226)
    Alpha=ScalarParameter() {defaultValue=1, group=Effect-ComicShading--Anime Maps, parameterName=EComicSh_Normal_Strength, sliderMax=1}
   )
   Alpha=NamedRerouteUsage()
  )
  n228 = NamedRerouteUsage()
  NamedRerouteDeclaration(None=StaticSwitch(True=LinearInterpolate(A=@n225, B=@n227, Alpha=@n228), False=@n225, Value=NamedRerouteUsage())) {name=NormalMap, nodeColor=<rgba 0 0.4375 1 1>}
  NamedRerouteDeclaration(None=Max(
    A=NamedRerouteDeclaration(None=@n63.R) {name=Msk_Prime, nodeColor=<rgba 0 1 0.7422 1>}
    B=NamedRerouteDeclaration(None=@n63.B) {name=Msk_Accent, nodeColor=<rgba 0 1 0.7656 1>}
   )) {name=Msk_Combine, nodeColor=<rgba 0 1 0.7656 1>}
  n229 = VectorParameter() {group=Effect-ComicShading--Uvs, parameterName=EComicSh_Offset_Vehicle, defaultValue=<rgba 0 0 0 1>}
  n230 = VectorParameter() {group=Effect-ComicShading--Uvs, parameterName=EComicSh_Offset_VehicleSide, defaultValue=<rgba 0 0 0 1>}
  n231 = MaterialFunctionCall(Float4=NamedRerouteUsage()) {materialFunction=&BreakOutFloat4Components}
  n232 = LinearInterpolate(
   A=LinearInterpolate(
    A=VectorParameter() {group=Effect-ComicShading--Uvs, parameterName=EComicSh_Offset_Weapon, defaultValue=<rgba 0 0 0 1>}.RGB
    B=LinearInterpolate(A=@n229.RGB, B=@n230.RGB, Alpha=@n231.G)
    Alpha=@n231.R
   )
   B=LinearInterpolate(
    A=VectorParameter() {group=Effect-ComicShading--Uvs, parameterName=EComicSh_Offset_Mech, defaultValue=<rgba 0 0 0 1>}.RGB
    B=VectorParameter() {group=Effect-ComicShading--Uvs, parameterName=EComicSh_Offset_Character, defaultValue=<rgba 0 0 0 1>}.RGB
    Alpha=@n231.A
   )
   Alpha=@n231.B
  )
  n233 = VectorParameter() {group=Effect-ComicShading--Uvs, parameterName=EComicSh_Offset_Backpack, defaultValue=<rgba 0 0 0 1>}
  n234 = MaterialFunctionCall(Float2=NamedRerouteUsage()) {materialFunction=&BreakOutFloat2Components}
  NamedRerouteDeclaration(None=LinearInterpolate(
    A=LinearInterpolate(A=@n232, B=@n233.RGB, Alpha=@n234.R)
    B=VectorParameter() {group=Effect-ComicShading--Uvs, parameterName=EComicSh_Offset_Pickaxe, defaultValue=<rgba 0 0 0 1>}.RGB
    Alpha=@n234.G
   )) {name=UVs_Offset, nodeColor=<rgba 0 0.625 1 1>}
  NamedRerouteDeclaration(None=@n31.Metallic) {name=Root_Metalic, nodeColor=<rgba 0.4062 0 1 1>}
  NamedRerouteDeclaration(None=@n31.Roughness) {name=Root_Roughness, nodeColor=<rgba 0.4062 0 1 1>}
  NamedRerouteDeclaration(None=@n63.A) {name=Msk_Gmap, nodeColor=<rgba 0 1 0.7656 1>}
  n235 = TextureObjectParameter() {group=Effect-ComicShading--Anime Maps, parameterName=EComicSh_BaseColor_Lit__Tex, texture=&T_WhitesRGB}
  NamedRerouteDeclaration(None=@n235) {name=Tex_Base_Lit, nodeColor=<rgba 0 0.5078 1 1>}
  n236 = TextureObjectParameter() {
   group=Effect-ComicShading--Anime Maps
   parameterName=EComicSh_BaseColor_Shaded__Tex
   texture=/Game/Global/Textures/Default/Blanks/T_MidgreysRGB
  }
  NamedRerouteDeclaration(None=@n236) {name=Tex_Base_Shaded, nodeColor=<rgba 0 0.5078 1 1>}
  n237 = TextureObjectParameter() {
   group=Effect-ComicShading--Anime Maps
   parameterName=EComicSh_Normal__Tex
   samplerType=SAMPLERTYPE_Normal
   texture=&FlatNormalBlue
  }
  NamedRerouteDeclaration(None=@n237) {name=Tex_Normal, nodeColor=<rgba 0 0.5078 1 1>}
  n238 = TextureObjectParameter() {
   group=Effect-ComicShading--Anime Maps
   parameterName=EComicSh_Mask__Tex
   samplerType=SAMPLERTYPE_Masks
   texture=&T_BlackMask
  }
  NamedRerouteDeclaration(None=@n238) {name=Tex_Mask, nodeColor=<rgba 0 0.5078 1 1>}
  n239 = MaterialFunctionCall(Color=Abs(None=NamedRerouteUsage())) {materialFunction=&SplitComponents}
  n240 = MaterialFunctionCall(Float4=NamedRerouteUsage()) {materialFunction=&BreakOutFloat4Components}
  n241 = LinearInterpolate(A=@n239.B, B=@n239.B, Alpha=@n240.G)
  n242 = NamedRerouteUsage()
  n243 = VectorParameter() {
   group=Effect-ComicShading--Anime Maps
   parameterName=EComicSh_Triplaner_Blend_Curvature_Amount
   sortPriority=0
   channelNames {a="", b="", g=Pits, r=Peaks}
   defaultValue=<rgba 2 2 0 1>
  }
  n244 = NamedRerouteUsage()
  n245 = Add(
   A=LinearInterpolate(A=LinearInterpolate(A=@n239.R, B=@n241, Alpha=@n240.R), B=@n239.G, Alpha=@n240.B)
   B=Add(A=Multiply(A=@n242, B=@n243.Peaks), B=Multiply(A=@n244, B=@n243.Pits))
  )
  n246 = ScalarParameter() {group=Effect-ComicShading--Anime Maps, parameterName=EComicSh_Triplaner_Contrast, sortPriority=0}
  NamedRerouteDeclaration(None=Max(
    A=MaterialFunctionCall(In=@n245, Contrast=@n246) {materialFunction=&CheapContrast}.Result
    B=ScalarParameter() {defaultValue=1, group=Effect-ComicShading--Anime Maps, parameterName="EComicSh_Triplaner  0=Enable", sortPriority=0}
   )) {name=Tri_Mask 01, nodeColor=<rgba 1 0.4688 0 1>}
  NamedRerouteUsage()
  NamedRerouteUsage()
  NamedRerouteUsage()
  PinBase() {reroutePins [
    Float3 {expression=&self:MaterialExpressionReroute_2}
    False {expression=&self:MaterialExpressionReroute_3}
    InputPin {expression=&self:MaterialExpressionReroute_5}
   ]}
  PinBase!(Output=@n50) {reroutePins [Output {expression=&self:MaterialExpressionReroute_4}]}
  PinBase!() {reroutePins [
    In {expression=&self:MaterialExpressionReroute_15}
    RampWidth {expression=&self:MaterialExpressionReroute_16}
   ]}
  MaterialExpressionReroute_15 = Reroute(None=@n27)
  n247 = Frac(None=@MaterialExpressionReroute_15.In)
  MaterialExpressionReroute_16 = Reroute(None=@n28)
  n248 = OneMinus(None=@MaterialExpressionReroute_16.RampWidth)
  n249 = Max(A=Divide(A=Subtract(A=@n247, B=@n248), B=@MaterialExpressionReroute_16.RampWidth), B=Constant() {r=0.5}) {desc=It is good to make functions look like a lancer}
  n250 = Add(A=Floor(None=@MaterialExpressionReroute_15.In), B=ConstantBiasScale(None=Cosine(None=@n249)))
  PinBase!(Output=@n250) {reroutePins [Output {expression=&self:MaterialExpressionReroute_17}]}
  PinBase!() {reroutePins [
    In {expression=&self:MaterialExpressionReroute_8}
    RampWidth {expression=&self:MaterialExpressionReroute_12}
   ]}
  PinBase!(Output=@n184) {reroutePins [Output {expression=&self:MaterialExpressionReroute_14}]}
  PixelNormalWS()
  PixelNormalWS()
  PixelNormalWS()
  MaterialExpressionReroute_17 = Reroute(None=@n250)
  n251 = Constant() {r=2}
  n252 = Constant() {r=1}
  Saturate(None=Add(A=Multiply(B=@n251), B=Subtract(B=@n252))) {desc=height blend with 1 heighmap}
  ScalarParameter() {defaultValue=1, group=Effect-ComicShading--HatchingLines, parameterName=EComicSh_Lines_Amount}
  ScalarParameter() {defaultValue=1.5, group=Effect-ComicShading--HatchingLines, parameterName=EComicSh_Hatching_Line_Amount, sliderMax=256}
  ScalarParameter() {defaultValue=0.5, group=Effect-ComicShading--DistanceFieldLines, parameterName=EComicSh_DF_Curvature_FallOff, sortPriority=0}
  ScalarParameter() {defaultValue=3, group=Effect-ComicShading--DistanceFieldLines, parameterName=EComicSh_DF_Curvature_Strength, sortPriority=0}
  ShadingModel() {shadingModel=MSM_Cloth}
  n253 = MaterialFunctionCall(Float3=@n68) {materialFunction=&BreakOutFloat3Components}
  n254 = ScalarParameter() {
   desc="0.0 = the light vector is locked on the XY plane.\r\n1.0 = the actual light vector is used  \r\n(for Faces use a blend value of 0,0)"
   group=01_ComicShading
   parameterName=LockedZLightVector_Blend
   sliderMax=1
  }
  n255 = MaterialFunctionCall(X=@n253.R, Y=@n253.R, Z=Multiply(A=@n253.B, B=@n254)) {materialFunction=&MakeFloat3}
  StaticSwitchParameter(True=LinearInterpolate(A=@n68, B=Normalize(VectorInput=@n255.Result), Alpha=VertexColor().A), False=@n68) {
   desc="Lock the Z axis of the Light Vector\r\nis driven by the VertexColor A channel (0 = Regular LightVector, 1= Locked LIghtVector)"
   group=01_ComicShading
   parameterName=UseLockedZLightVector
   sortPriority=0
  }
  StaticSwitch(True=@n95, False=@n95, Value=NamedRerouteUsage())
  TextureCoordinate()
  TextureCoordinate()
  Transform() {transformSourceType=TRANSFORMSOURCE_World, transformType=TRANSFORM_Tangent}
  VertexNormalWS()
 ]}
