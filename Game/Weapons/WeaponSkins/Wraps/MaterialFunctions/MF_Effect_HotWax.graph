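# Material function for weapon wraps that adds an animated, lava-lamp style hot wax effect to an incoming material.
# Two blob layers, procedural or textured, move by per-cell seeded scroll or sine over side-projected UVs, with noise or texture warp.
# Layers combine with gradients and optional noise, shaped through a curve atlas.
# Parameters vary by item type; CM and decal masks limit the blend and feed a mask output.
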MaterialFunction /Game/Weapons/WeaponSkins/Wraps/MaterialFunctions/MF_Effect_HotWax
graph {nodes [
  n1 = Multiply(A=NamedRerouteUsage(), B=ScalarParameter() {group=Effect_HotWax_Object, parameterName=EHWax_ObjectA_Warp_Strength})
  n2 = VectorParameter() {group=Effect_HotWax_Object, parameterName=EHWax_ObjectA_Offset, defaultValue=<rgba 0 0 0 1>}
  n3 = Add(A=ComponentMask(None=@n2.RGB) {g=true, r=true}, B=NamedRerouteUsage())
  n4 = ComponentMask(None=@n3) {g=true}
  n5 = NamedRerouteUsage()
  n6 = Constant() {r=0.5}
  n7 = Subtract(A=Add(A=Multiply(A=@n4, B=@n5), B=@n6), B=Multiply(A=@n6, B=@n5)) {desc=Re offset to maintain center}
  n8 = Multiply(A=ComponentMask(None=@n3) {r=true}, B=@n5)
  n9 = VectorParameter() {group=Effect_HotWax_Object_A, parameterName=EHWax_ObjectA_Seed, sortPriority=200, defaultValue=<rgba 0.1253 0.0235 0 0>}
  n10 = ComponentMask(None=@n9.RGB) {g=true, r=true}
  n11 = MaterialFunctionCall!(2 Vector=Add(A=Floor(None=@n8), B=@n10), RandomSeed) {materialFunction=&/MF_Random_2}
  n12 = Time()
  n13 = ScalarParameter() {defaultValue=-0.1, group=Effect_HotWax_Object_A, parameterName=EHWax_ObjectA_Speed_Sine}
  n14 = Sine(None=Add(A=@n11.utput, B=Multiply(A=@n12, B=@n13)))
  n15 = NamedRerouteUsage()
  n16 = Saturate(None=Add(A=@n7, B=Multiply(A=@n14, B=@n15)))
  n17 = AppendVector(A=@n8, B=@n7)
  n18 = VectorParameter() {
   group=Effect_HotWax_Object_A
   parameterName=EHWax_ObjectA_Speed_Range
   channelNames {a="", b="", g=High, r=Low}
   defaultValue=<rgba 0.1 1 0 0>
  }
  n19 = Time()
  n20 = VectorParameter() {group=Effect_HotWax_Object_A, parameterName=EHWax_ObjectA_Speed XY, defaultValue=<rgba 0 -1 0 0>}
  n21 = Multiply(
   A=AppendVector(A=Constant() {r=1}, B=LinearInterpolate(A=@n18.Low, B=@n18.Low, Alpha=@n11.utput) {constA=0.1})
   B=ComponentMask(None=Multiply(A=@n19, B=@n20.RGB)) {g=true, r=true}
  )
  n22 = Frac(None=StaticSwitch(True=AppendVector(A=@n8, B=@n16), False=Add(A=@n17, B=@n21), Value=NamedRerouteUsage()))
  n23 = Add(A=@n1, B=@n22)
  n24 = StaticSwitchParameter(True=NamedRerouteUsage(), False=NamedRerouteUsage()) {desc=Look To TopSide Shared, group=Effect_HotWax_Object, parameterName="EHWax__MSk_Tex_ObjectA-->TopS_00B_Sharerd_Tex_Msk", sortPriority=5}
  n25 = StaticSwitchParameter(True=NamedRerouteUsage(), False=@n24) {desc=Look To TopSide Shared, group=Effect_HotWax_Object, parameterName="EHWax__MSk_Tex_ObjectA -->TopS_00A_Sharerd_Tex_Msk", sortPriority=5}
  n26 = NamedRerouteUsage()
  n27 = Multiply(
   A=TextureProperty(Tex=@n25)
   B=ScalarParameter() {defaultValue=1, group=Effect_HotWax_Object, parameterName=EHWax_ObjectS_MipBias, sortPriority=2}
  )
  n28 = ChannelMaskParameter(None=TextureSample!(
    UVs=Saturate(None=@n23)
    Tex=@n25
    Level=MaterialFunctionCall!(UVs=@n26, Texture Size=@n27) {materialFunction=/Engine/Functions/Engine_MaterialFunctions01/Texturing/ComputeMipLevel}.MipLevel
    Apply View MipBias
   ) {mipValueMode=TMVM_MipLevel}.RGBA) {group=Effect_HotWax_Object, parameterName=EHWax_ObjectA_Channel_Tex_Msk}
  n29 = Add(A=AppendVector(A=@n8, B=@n7), B=@n21) {constB=0}
  n30 = MaterialFunctionCall(2 Vector=Add(A=Floor(None=@n29), B=@n10), RandomSeed=Constant() {r=1.123}) {materialFunction=&MF_Random_2}
  n31 = Round(None=Saturate(None=LinearInterpolate(Alpha=@n30.utput) {constA=-2, constB=0.75}))
  n32 = NamedRerouteUsage()
  n33 = StaticSwitch(True=@n28.R, False=Multiply(A=@n28.R, B=@n31), Value=@n32)
  n34 = Constant() {r=0.5}
  n35 = ScalarParameter() {defaultValue=0.5, group=Effect_HotWax_Object_A, parameterName=EHWax_ObjectA_Scale}
  n36 = Power(
   Base=Add(A=@n1, B=SphereMask(A=@n22, B=@n34, Radius=@n35) {hardnessPercent=0})
   Exp=ScalarParameter() {defaultValue=3, group=Effect_HotWax_Object_A, parameterName=EHWax_ObjectA_Falloff}
  )
  n37 = StaticSwitch(True=@n36, False=Multiply(A=@n36, B=@n31), Value=@n32)
  n38 = Multiply(A=NamedRerouteUsage(), B=ScalarParameter() {group=Effect_HotWax_Object, parameterName=EHWax_ObjectB_Warp_Strength})
  n39 = VectorParameter() {group=Effect_HotWax_Object, parameterName=EHWax_ObjectB_Offset, defaultValue=<rgba 0.5 0 0 1>}
  n40 = Add(A=ComponentMask(None=@n39.RGB) {g=true, r=true}, B=NamedRerouteUsage())
  n41 = ComponentMask(None=@n40) {g=true}
  n42 = NamedRerouteUsage()
  n43 = Constant() {r=0.5}
  n44 = Subtract(A=Add(A=Multiply(A=@n41, B=@n42), B=@n43), B=Multiply(A=@n43, B=@n42)) {desc=Re offset to maintain center}
  n45 = Multiply(A=ComponentMask(None=@n40) {r=true}, B=@n42)
  n46 = VectorParameter() {group=Effect_HotWax_Object_B, parameterName=EHWax_ObjectB_Seed, sortPriority=200, defaultValue=<rgba 5 3 0 0>}
  n47 = ComponentMask(None=@n46.RGB) {g=true, r=true}
  n48 = MaterialFunctionCall(2 Vector=Add(A=Floor(None=@n45), B=@n47)) {materialFunction=&MF_Random_2}
  n49 = Time()
  n50 = ScalarParameter() {defaultValue=-0.15, group=Effect_HotWax_Object_B, parameterName=EHWax_ObjectB_Speed_Sine}
  n51 = Sine(None=Add(A=@n48.utput, B=Multiply(A=@n49, B=@n50)))
  n52 = NamedRerouteUsage()
  n53 = Saturate(None=Add(A=@n44, B=Multiply(A=@n51, B=@n52)))
  n54 = AppendVector(A=@n45, B=@n44)
  n55 = VectorParameter() {
   group=Effect_HotWax_Object_B
   parameterName=EHWax_ObjectB_Speed_Range
   channelNames {a="", b="", g=High, r=Low}
   defaultValue=<rgba 0.1 1 0 0>
  }
  n56 = VectorParameter() {group=Effect_HotWax_Object_B, parameterName=EHWax_ObjectB_Speed XY, defaultValue=<rgba 0 -1 0 0>}
  n57 = Multiply(
   A=AppendVector(A=Constant() {r=1}, B=LinearInterpolate(A=@n55.Low, B=@n55.Low, Alpha=@n48.utput) {constA=0.1})
   B=Multiply(A=Time(), B=ComponentMask(None=@n56.RGB) {g=true, r=true})
  )
  n58 = Frac(None=StaticSwitch(True=AppendVector(A=@n45, B=@n53), False=Add(A=@n54, B=@n57), Value=NamedRerouteUsage()))
  n59 = Add(A=@n38, B=@n58)
  n60 = StaticSwitchParameter(True=NamedRerouteUsage(), False=NamedRerouteUsage()) {desc=Look To TopSide Shared, group=Effect_HotWax_Object, parameterName="EHWax__MSk_Tex_ObjectB-->TopS_00B_Sharerd_Tex_Msk", sortPriority=6}
  n61 = StaticSwitchParameter(True=NamedRerouteUsage(), False=@n60) {desc=Look To TopSide Shared, group=Effect_HotWax_Object, parameterName="EHWax__MSk_Tex_ObjectB -->TopS_00A_Sharerd_Tex_Msk", sortPriority=6}
  n62 = NamedRerouteUsage()
  n63 = Multiply(
   A=TextureProperty(Tex=@n61)
   B=ScalarParameter() {defaultValue=1, group=Effect_HotWax_Object, parameterName=EHWax_ObjectS_MipBias, sortPriority=2}
  )
  n64 = ChannelMaskParameter(None=TextureSample(
    UVs=Saturate(None=@n59)
    Tex=@n61
    Level=MaterialFunctionCall(UVs=@n62, Texture Size=@n63) {materialFunction=&ComputeMipLevel}.MipLevel
   ) {mipValueMode=TMVM_MipLevel}.RGBA) {group=Effect_HotWax_Object, parameterName=EHWax_ObjectB_Channel_Tex_Msk}
  n65 = Add(A=AppendVector(A=@n45, B=@n44), B=@n57) {constB=0}
  n66 = MaterialFunctionCall(2 Vector=Add(A=Floor(None=@n65), B=@n47), RandomSeed=Constant() {r=1.123}) {materialFunction=&MF_Random_2}
  n67 = Round(None=Saturate(None=LinearInterpolate(Alpha=@n66.utput) {constA=-2, constB=0.75}))
  n68 = NamedRerouteUsage()
  n69 = StaticSwitch(True=@n64.R, False=Multiply(A=@n64.R, B=@n67), Value=@n68)
  n70 = Constant() {r=0.5}
  n71 = ScalarParameter() {defaultValue=0.5, group=Effect_HotWax_Object_B, parameterName=EHWax_ObjectB_Scale}
  n72 = Power(
   Base=Add(A=@n38, B=SphereMask(A=@n58, B=@n70, Radius=@n71) {hardnessPercent=0})
   Exp=ScalarParameter() {defaultValue=3, group=Effect_HotWax_Object_B, parameterName=EHWax_ObjectB_Falloff}
  )
  n73 = StaticSwitch(True=@n72, False=Multiply(A=@n72, B=@n67), Value=@n68)
  Add(
   A=NamedRerouteDeclaration(None=StaticSwitchParameter(True=@n33, False=@n37) {group=Effect_HotWax_Object, parameterName="EHWax_ObjectS_procedural --------->Tex", sortPriority=0}) {name=ObjectA, nodeColor=<rgba 0 1 0.6719 1>}
   B=NamedRerouteDeclaration(None=StaticSwitchParameter(True=@n69, False=@n73) {group=Effect_HotWax_Object, parameterName="EHWax_ObjectS_procedural --------->Tex", sortPriority=0}) {name=ObjectB, nodeColor=<rgba 0 1 0.6719 1>}
  )
  n74 = NamedRerouteUsage()
  n75 = NamedRerouteUsage()
  n76 = NamedRerouteUsage()
  n77 = VectorParameter() {
   group=Effect_HotWax_Object
   parameterName=EHWax_ObjectB_FadeGradent
   channelNames {a="0=Use", b=Power, g=Falloff, r=Offset}
   defaultValue=<rgba -1 -0.5 1 1>
  }
  n78 = NamedRerouteUsage()
  n79 = Add(
   A=@n77.Offset
   B=StaticSwitchParameter(True=NamedRerouteUsage(), False=ComponentMask(None=@n78) {g=true}) {group=Effect_HotWax_Object, parameterName=EHWax_ObjectA_B_Fade Use Base Gradint Direction}
  )
  n80 = Saturate(None=Power(Base=Divide(A=@n79, B=@n77.Falloff), Exp=@n77.Power))
  n81 = NamedRerouteUsage()
  n82 = Multiply(A=Multiply(A=StaticSwitch(True=@n74, False=@n75, Value=@n76), B=Max(A=@n77."0=Use", B=@n80)), B=@n81)
  n83 = NamedRerouteUsage()
  n84 = NamedRerouteUsage()
  n85 = VectorParameter() {
   group=Effect_HotWax_Object
   parameterName=EHWax_ObjectA_FadeGradent
   channelNames {a="0=Use", b=Power, g=Falloff, r=Offset}
   defaultValue=<rgba -1 -0.5 1 1>
  }
  n86 = NamedRerouteUsage()
  n87 = Add(
   A=@n85.Offset
   B=StaticSwitchParameter(True=NamedRerouteUsage(), False=ComponentMask(None=@n86) {g=true}) {group=Effect_HotWax_Object, parameterName=EHWax_ObjectA_B_Fade Use Base Gradint Direction}
  )
  n88 = Saturate(None=Power(Base=Divide(A=@n87, B=@n85.Falloff), Exp=@n85.Power))
  n89 = Multiply(A=@n81, B=Multiply(A=StaticSwitch(True=@n83, False=@n84, Value=@n76), B=Max(A=@n85."0=Use", B=@n88)))
  n90 = NamedRerouteUsage()
  n91 = OneMinus(None=@n89)
  n92 = StaticSwitchParameter(True=Add(A=@n89, B=Multiply(A=@n90, B=@n91)), False=@n90) {defaultValue=true, group=Effect_HotWax_Object, parameterName=EHWax_ Add ObjectsA, sortPriority=0}
  n93 = OneMinus(None=@n82)
  n94 = StaticSwitchParameter(True=Add(A=@n82, B=Multiply(A=@n92, B=@n93)), False=@n92) {defaultValue=true, group=Effect_HotWax_Object, parameterName=EHWax_Add ObjectsB, sortPriority=0}
  n95 = SmoothStep(Value=@n90) {constMax=0.6, constMin=0.1}
  n96 = Multiply(A=NamedRerouteUsage(), B=@n95)
  n97 = LinearInterpolate(A=@n94, Alpha=@n96) {constB=0}
  n98 = Power(
   Base=StaticSwitchParameter(
    True=StaticSwitchParameter(True=@n94, False=@n97) {group=Effect_HotWax_Noise, parameterName=EHWax_Add Noise_AfterShape, sortPriority=0}
    False=@n94
   ) {group=Effect_HotWax_Noise, parameterName=EHWax_Add Noise, sortPriority=0}
   Exp=ScalarParameter() {defaultValue=2, group=Effect_HotWax, parameterName=EHWax_CC_Shape_Falloff}
  )
  n99 = CurveAtlasRowParameter(CurveTime=@n98) {
   atlas=/Game/Weapons/WeaponSkins/Wraps/Textures/CruveAtlas_Wraps_01
   curve=/Game/Weapons/WeaponSkins/Wraps/Textures/LavaLamp/CC_Liquid_HotWax_Shape
   defaultValue=180
   group=Effect_HotWax
   parameterName=EHWax_CC_Shape
  }
  AppendVector(A=@n99, B=@n99)
  Constant() {r=1}
  Constant() {r=0.4}
  Constant() {r=0.4}
  n100 = NamedRerouteUsage()
  n101 = NamedRerouteUsage()
  n102 = ChannelMaskParameter(None=FunctionInput() {inputName=Decal_Mask_RGBA, inputType=FunctionInput_Vector4, sortPriority=90}) {group="Effect_HotWax  Mask", parameterName=EHWax_Decal_Mask_RGBA_Channel}
  n103 = FunctionInput() {inputName=Decal_Mask, inputType=FunctionInput_Scalar, sortPriority=90}
  n104 = SmoothStep(
   Min=ScalarParameter() {group="Effect_HotWax  Mask", parameterName="EHWax_Decal_Alpha Masked  Contrast", sortPriority=50}
   Value=LinearInterpolate(
    B=StaticSwitchParameter(True=@n102.R, False=@n103) {group="Effect_HotWax  Mask", parameterName=EHWax_Decal_Mask_RGBA, sortPriority=50}
    Alpha=ScalarParameter() {defaultValue=1, group="Effect_HotWax  Mask", parameterName="EHWax_Decal_Alpha Masked  1=Active", sortPriority=50}
   )
  )
  n105 = TextureSample!(UVs, Tex, Apply View MipBias) {samplerType=SAMPLERTYPE_Masks, texture=/Game/Weapons/FORT_Rifles/Textures/T_Scar_CM}
  n106 = MaterialFunctionCall(Float4=FunctionInput(Preview=@n105.RGBA) {inputName=CM_Msk, inputType=FunctionInput_Vector4, sortPriority=4}) {materialFunction=&BreakOutFloat4Components}
  n107 = NamedRerouteDeclaration(None=Min(A=ScalarParameter() {defaultValue=1, group="Effect_HotWax  Mask", parameterName="EHWax_Mask_Primary   1=On"}, B=@n106.R)) {name=Msk_Prime, nodeColor=<rgba 0 1 0.7422 1>}
  n108 = NamedRerouteDeclaration(None=Min(A=@n106.B, B=ScalarParameter() {group="Effect_HotWax  Mask", parameterName="EHWax_Mask_Accent   1=On"})) {name=Msk_Accent, nodeColor=<rgba 0 1 0.7656 1>}
  n109 = Min(
   A=StaticSwitchParameter(True=OneMinus(None=@n104), False=Constant() {r=1}) {group=Customization, parameterName=Function_Left_Right_Decal_V2}
   B=Max(A=Max(A=@n107, B=@n108), B=ScalarParameter() {group="Effect_HotWax  Mask", parameterName="EHWax_Mask_Disable   1=On"})
  )
  n110 = Min(A=@n109, B=NamedRerouteUsage())
  n111 = SetMaterialAttributes!(
   MaterialAttributes
   Normal=TextureSample() {samplerType=SAMPLERTYPE_Normal, texture=/Game/Weapons/FORT_Rifles/T_SCAR_N}.RGB
   BaseColor=Constant3Vector() {constant=<rgba 1 0 0 0>}
  ) {attributeSetTypes [#1, #2]}
  n112 = GetMaterialAttributes(None=FunctionInput(Preview=@n111) {inputName=Material, inputType=FunctionInput_MaterialAttributes}) {attributeGetTypes [
    #2
    #3
    #1
    #4
   ]}
  FunctionOutput(None=StaticSwitchParameter(
    True=BlendMaterialAttributes(A=@n100, B=@n101, Alpha=@n110) {vertexAttributeBlendType=UseA}
    False=NamedRerouteDeclaration(None=@n112.MaterialAttributes) {name=Root_Material, nodeColor=<rgba 0.7812 1 0 1>}
   ) {group=Customization, parameterName=Effect_HotWax})
  FunctionOutput(None=@n110) {outputName=Hotwax_Msk}
  MaterialFunctionCall(Float2=NamedRerouteUsage()) {materialFunction=&BreakOutFloat2Components}
  MaterialFunctionCall(Float2=NamedRerouteUsage()) {materialFunction=&BreakOutFloat2Components}
  MaterialFunctionCall(Float2=NamedRerouteUsage()) {materialFunction=&BreakOutFloat2Components}
  MaterialFunctionCall!(Height Map, Normal Map Intensity, Height Map UV Offset, Coordinates, Height Map Channel Selector) {materialFunction=/Engine/Functions/Engine_MaterialFunctions03/Procedurals/NormalFromHeightmap}
  MaterialFunctionCall(Float2=NamedRerouteUsage()) {materialFunction=&BreakOutFloat2Components}
  n113 = Constant()
  n114 = StaticSwitchParameter(True=@n102.R, False=@n113) {group="Effect_HotWax  Mask", parameterName=EHWax_Decal_Mask_RGBA, sortPriority=50}
  Max(
   A=StaticSwitchParameter(
    True=StaticSwitchParameter(True=@n103, False=@n114) {group="Effect_HotWax  Mask", parameterName=EHWax_Decal_Mask_Alpha, sortPriority=50}
    False=@n113
   ) {group=Customization, parameterName=Function_Left_Right_Decal_V2}
   B=@n109
  )
  n115 = MaterialFunctionCall(Float4=NamedRerouteUsage()) {materialFunction=&BreakOutFloat4Components}
  n116 = LinearInterpolate(
   A=ScalarParameter() {defaultValue=2, group=Effect_HotWax_Object_A, parameterName=EHWax_ObjectA_Sine_Strength_Weapon, sortPriority=100}
   B=LinearInterpolate(
    A=ScalarParameter() {defaultValue=1, group=Effect_HotWax_Object_A, parameterName=EHWax_ObjectA_Sine_Strength_Vehicle, sortPriority=100}
    B=ScalarParameter() {defaultValue=1, group=Effect_HotWax_Object_A, parameterName=EHWax_ObjectA_Sine_Strength_SideVehicle, sortPriority=100}
    Alpha=@n115.G
   )
   Alpha=@n115.R
  )
  n117 = LinearInterpolate(
   A=ScalarParameter() {defaultValue=1, group=Effect_HotWax_Object_A, parameterName=EHWax_ObjectA_Sine_Strength_Mech, sortPriority=100}
   B=ScalarParameter() {defaultValue=1, group=Effect_HotWax_Object_A, parameterName=EHWax_ObjectA_Sine_Strength_Character, sortPriority=100}
   Alpha=@n115.A
  )
  n118 = MaterialFunctionCall(Float2=NamedRerouteUsage()) {materialFunction=&BreakOutFloat2Components}
  n119 = LinearInterpolate(
   A=LinearInterpolate(
    A=LinearInterpolate(A=@n116, B=@n117, Alpha=@n115.B)
    B=ScalarParameter() {defaultValue=1, group=Effect_HotWax_Object_A, parameterName=EHWax_ObjectA_Sine_Strength_Backpack, sortPriority=100}
    Alpha=@n118.R
   )
   B=ScalarParameter() {defaultValue=1, group=Effect_HotWax_Object_A, parameterName=EHWax_ObjectA_Sine_Strength_Pickaxe, sortPriority=100}
   Alpha=@n118.G
  )
  Multiply(A=@n119) {constB=-1}
  n120 = MaterialFunctionCall(Float4=NamedRerouteUsage()) {materialFunction=&BreakOutFloat4Components}
  n121 = LinearInterpolate(
   A=ScalarParameter() {defaultValue=2, group=Effect_HotWax_Object_B, parameterName=EHWax_ObjectB_Sine_Strength_Weapon, sortPriority=100}
   B=LinearInterpolate(
    A=ScalarParameter() {defaultValue=1, group=Effect_HotWax_Object_B, parameterName=EHWax_ObjectB_Sine_Strength_Vehicle, sortPriority=100}
    B=ScalarParameter() {defaultValue=1, group=Effect_HotWax_Object_B, parameterName=EHWax_ObjectB_Sine_Strength_SideVehicle, sortPriority=100}
    Alpha=@n120.G
   )
   Alpha=@n120.R
  )
  n122 = LinearInterpolate(
   A=ScalarParameter() {defaultValue=1, group=Effect_HotWax_Object_B, parameterName=EHWax_ObjectB_Sine_Strength_Mech, sortPriority=100}
   B=ScalarParameter() {defaultValue=1, group=Effect_HotWax_Object_B, parameterName=EHWax_ObjectB_Sine_Strength_Character, sortPriority=100}
   Alpha=@n120.A
  )
  n123 = MaterialFunctionCall(Float2=NamedRerouteUsage()) {materialFunction=&BreakOutFloat2Components}
  n124 = LinearInterpolate(
   A=LinearInterpolate(
    A=LinearInterpolate(A=@n121, B=@n122, Alpha=@n120.B)
    B=ScalarParameter() {defaultValue=1, group=Effect_HotWax_Object_B, parameterName=EHWax_ObjectB_Sine_Strength_Backpack, sortPriority=100}
    Alpha=@n123.R
   )
   B=ScalarParameter() {defaultValue=1, group=Effect_HotWax_Object_B, parameterName=EHWax_ObjectB_Sine_Strength_Pickaxe, sortPriority=100}
   Alpha=@n123.G
  )
  Multiply(A=@n124) {constB=-1}
  NamedRerouteDeclaration(None=TextureObjectParameter() {group=Effect_HotWax, parameterName=EHWax__MSk_Tex, texture=/Game/Weapons/WeaponSkins/Wraps/Textures/HotWax/T_Wrap_HotWax_MSk}) {name=Tex_Grad, nodeColor=<rgba 1 0.4453 0 1>}
  NamedRerouteDeclaration(None=@n106.G) {name=Msk_Curvature, nodeColor=<rgba 0 1 0.7422 1>}
  n125 = StaticBoolParameter() {group=Effect_HotWax_Gradent, parameterName=EHWax_Grad_Mirror}
  NamedRerouteDeclaration(None=@n125) {name=Grad_Mirror_State, nodeColor=<rgba 0.5703 0 1 1>}
  n126 = ScalarParameter() {defaultValue=0.1, group=Effect_HotWax_Gradent, parameterName=EHWax_Grad_Spread_Weapon, sortPriority=150}
  n127 = MaterialFunctionCall(Float4=NamedRerouteUsage()) {materialFunction=&BreakOutFloat4Components}
  n128 = LinearInterpolate(
   A=ScalarParameter() {group=Effect_HotWax_Gradent, parameterName=EHWax_Grad_Spread_Vehicle, sortPriority=150}
   B=ScalarParameter() {group=Effect_HotWax_Gradent, parameterName=EHWax_Grad_Spread_SideVehicle, sortPriority=150}
   Alpha=@n127.G
  )
  n129 = ScalarParameter() {group=Effect_HotWax_Gradent, parameterName=EHWax_Grad_Spread_Mech, sortPriority=150}
  n130 = ScalarParameter() {group=Effect_HotWax_Gradent, parameterName=EHWax_Grad_Spread_Character, sortPriority=150}
  n131 = MaterialFunctionCall(Float2=NamedRerouteUsage()) {materialFunction=&BreakOutFloat2Components}
  n132 = LinearInterpolate(
   A=LinearInterpolate(A=LinearInterpolate(A=@n126, B=@n128, Alpha=@n127.R), B=LinearInterpolate(A=@n129, B=@n130, Alpha=@n127.A), Alpha=@n127.B)
   B=ScalarParameter() {group=Effect_HotWax_Gradent, parameterName=EHWax_Grad_Spread_Backpack, sortPriority=150}
   Alpha=@n131.R
  )
  n133 = ScalarParameter() {group=Effect_HotWax_Gradent, parameterName=EHWax_Grad_Spread_Pickaxe, sortPriority=150}
  NamedRerouteDeclaration(None=Multiply(A=LinearInterpolate(A=@n132, B=@n133, Alpha=@n131.G)) {constB=-1}) {name=UVs_Grad_Spread, nodeColor=<rgba 0 0.9125 1 1>}
  n134 = ScalarParameter() {defaultValue=1, group=Effect_HotWax_Gradent, parameterName=EHWax_Grad_FallOff_Vehicle, sortPriority=150}
  n135 = ScalarParameter() {defaultValue=1, group=Effect_HotWax_Gradent, parameterName=EHWax_Grad_FallOff_SideVehicle, sortPriority=150}
  n136 = MaterialFunctionCall(Float4=NamedRerouteUsage()) {materialFunction=&BreakOutFloat4Components}
  n137 = LinearInterpolate(
   A=LinearInterpolate(
    A=ScalarParameter() {defaultValue=1, group=Effect_HotWax_Gradent, parameterName=EHWax_Grad_FallOff_Weapon, sortPriority=150}
    B=LinearInterpolate(A=@n134, B=@n135, Alpha=@n136.G)
    Alpha=@n136.R
   )
   B=LinearInterpolate(
    A=ScalarParameter() {defaultValue=1, group=Effect_HotWax_Gradent, parameterName=EHWax_Grad_FallOff_Mech, sortPriority=150}
    B=ScalarParameter() {defaultValue=1, group=Effect_HotWax_Gradent, parameterName=EHWax_Grad_FallOff_Character, sortPriority=150}
    Alpha=@n136.A
   )
   Alpha=@n136.B
  )
  n138 = ScalarParameter() {defaultValue=1, group=Effect_HotWax_Gradent, parameterName=EHWax_Grad_FallOff_Backpack, sortPriority=150}
  n139 = MaterialFunctionCall(Float2=NamedRerouteUsage()) {materialFunction=&BreakOutFloat2Components}
  NamedRerouteDeclaration(None=LinearInterpolate(
    A=LinearInterpolate(A=@n137, B=@n138, Alpha=@n139.R)
    B=ScalarParameter() {defaultValue=1, group=Effect_HotWax_Gradent, parameterName=EHWax_Grad_FallOff_Pickaxe, sortPriority=150}
    Alpha=@n139.G
   )) {name=UVs_Grad_Falloff, nodeColor=<rgba 0 0.9125 1 1>}
  n140 = VectorParameter() {group=Effect_HotWax_Warp, parameterName=EHWax_WarpScale_Weapon, defaultValue=<rgba 1 1 1 4>}
  n141 = VectorParameter() {group=Effect_HotWax_Warp, parameterName=EHWax_WarpScale_Vehicle, defaultValue=<rgba 1 1 1 4>}
  n142 = Multiply(A=@n141.RGB, B=@n141.RGB)
  n143 = VectorParameter() {group=Effect_HotWax_Warp, parameterName=EHWax_WarpScale_VehicleSide, defaultValue=<rgba 1 1 1 4>}
  n144 = Multiply(A=@n143.RGB, B=@n143.RGB)
  n145 = MaterialFunctionCall(Float4=NamedRerouteUsage()) {materialFunction=&BreakOutFloat4Components}
  n146 = VectorParameter() {group=Effect_HotWax_Warp, parameterName=EHWax_WarpScale_Mech, defaultValue=<rgba 1 1 1 4>}
  n147 = VectorParameter() {group=Effect_HotWax_Warp, parameterName=EHWax_WarpScale_Character, defaultValue=<rgba 1 1 1 4>}
  n148 = LinearInterpolate(
   A=LinearInterpolate(A=Multiply(A=@n140.RGB, B=@n140.RGB), B=LinearInterpolate(A=@n142, B=@n144, Alpha=@n145.G), Alpha=@n145.R)
   B=LinearInterpolate(A=Multiply(A=@n146.RGB, B=@n146.RGB), B=Multiply(A=@n147.RGB, B=@n147.RGB), Alpha=@n145.A)
   Alpha=@n145.B
  )
  n149 = VectorParameter() {group=Effect_HotWax_Warp, parameterName=EHWax_WarpScale_Backpack, defaultValue=<rgba 1 1 1 4>}
  n150 = Multiply(A=@n149.RGB, B=@n149.RGB)
  n151 = MaterialFunctionCall(Float2=NamedRerouteUsage()) {materialFunction=&BreakOutFloat2Components}
  n152 = VectorParameter() {group=Effect_HotWax_Warp, parameterName=EHWax_WarpScale_Pickaxe, defaultValue=<rgba 1 1 1 4>}
  NamedRerouteDeclaration(None=LinearInterpolate(A=LinearInterpolate(A=@n148, B=@n150, Alpha=@n151.R), B=Multiply(A=@n152.RGB, B=@n152.RGB), Alpha=@n151.G)) {name=UVs_Warp_Scale, nodeColor=<rgba 0 0.1613 1 1>}
  n153 = NamedRerouteUsage()
  n154 = NamedRerouteUsage()
  n155 = NamedRerouteUsage()
  n156 = Time()
  n157 = VectorParameter() {group=Effect_HotWax_Warp, parameterName=EHWax_Warp_Speed, channelNames {a="", b="", g="", r=Foward}, defaultValue=<rgba 0.2 0.2 0.2 0>}
  n158 = Multiply(A=@n156, B=@n157)
  n159 = NamedRerouteUsage()
  n160 = StaticBool() {value=true}
  n161 = StaticSwitchParameter(True=NamedRerouteUsage(), False=NamedRerouteUsage()) {desc=Look To TopSide Shared, group=Effect_HotWax_Warp, parameterName="EHWax__MSk_Tex_Warp-->TopS_00B_Sharerd_Tex_Msk", sortPriority=5}
  n162 = StaticSwitchParameter(True=NamedRerouteUsage(), False=@n161) {desc=Look To TopSide Shared, group=Effect_HotWax_Warp, parameterName="EHWax__MSk_Tex_Warp -->TopS_00A_Sharerd_Tex_Msk", sortPriority=5}
  n163 = ChannelMaskParameter(None=TextureSample(
    UVs=MaterialFunctionCall!(
     Normal_Map
     Offset Rt
     Pivot
     Caculated_Bounds=@n153
     Pre_Skinned_Normals=@n154
     is_State=@n155
     UV_Rotation
     Uvs_Offset=@n158
     Uvs_TileRate=@n159
     Uvs_Offset_Pivot
     For_Safe_Zones
     Rotation_Comp
     Clamp Y
     Clamp X
     Clamp Padding_Comp
     No_Rotation=@n160
    ) {materialFunction=&/MF_Projection_Side_Normal_V2}.UVs
    Tex=@n162
   ).RGBA) {group=Effect_HotWax_Warp, maskChannel=Alpha, parameterName=EHWax_Warp_Channel_Tex_Msk, sortPriority=1, defaultValue=<rgba 0 0 0 1>}
  n164 = Add(A=@n163.R) {constB=-0.5}
  n165 = Add(A=Multiply(A=NamedRerouteUsage(), B=@n159), B=Multiply(A=@n156, B=NamedRerouteUsage()))
  n166 = ChannelMaskParameter(None=Multiply(A=VectorNoise(World Position=@n165) {noiseFunction=VNF_VectorALU}) {constB=0.5}) {group=Effect_HotWax_Warp, parameterName=EHWax_Warp_Noise_Channel, sortPriority=1}
  n167 = Add(A=NamedRerouteUsage()) {constB=-0.5}
  n168 = ScalarParameter() {group=Effect_HotWax_Gradent_Uv, parameterName=EHWax_Curavature_Warp_Strength, sortPriority=1}
  NamedRerouteDeclaration(None=Add(
    A=StaticSwitchParameter(True=@n164, False=@n166.R) {group=Effect_HotWax_Warp, parameterName="EHWax_Warp_procedural --------->MSk_Tex", sortPriority=0}
    B=Multiply(A=@n167, B=@n168)
   )) {name=Warp, nodeColor=<rgba 0 0.3906 1 1>}
  n169 = ScalarParameter() {group=Effect_HotWax_Gradent_Uv, parameterName="EHWax_Grad_Rotation_Weapon\\Pickaxe", sortPriority=150}
  n170 = MaterialFunctionCall(Float4=NamedRerouteUsage()) {materialFunction=&BreakOutFloat4Components}
  n171 = LinearInterpolate(
   A=ScalarParameter() {group=Effect_HotWax_Gradent_Uv, parameterName=EHWax_Grad_Rotation_Vehicle, sortPriority=150}
   B=ScalarParameter() {group=Effect_HotWax_Gradent_Uv, parameterName=EHWax_Grad_Rotation_Vehicle_Side, sortPriority=150}
   Alpha=@n170.G
  )
  NamedRerouteDeclaration(None=LinearInterpolate(
    A=LinearInterpolate(A=@n169, B=@n171, Alpha=@n170.R)
    B=ScalarParameter() {group=Effect_HotWax_Gradent_Uv, parameterName="EHWax_Grad_Rotation_Mech\\Character\\Back", sortPriority=150}
    Alpha=@n170.B
   )) {name=UVs_Grad_Rotation, nodeColor=<rgba 0 0.9125 1 1>}
  n172 = TextureObjectParameter() {
   group=TopSide_V02
   parameterName=TopS_00B_Sharerd_Tex_Msk
   samplerType=SAMPLERTYPE_Masks
   sortPriority=249
   texture=&T_BlackMask
  }
  NamedRerouteDeclaration(None=@n172) {name=Tex_00B_Shared_Tex, nodeColor=<rgba 1 0.8906 0 1>}
  n173 = VectorParameter() {
   group=Effect_HotWax_Warp
   parameterName="EHWax_Warp_Speed_Weapon\\Pickaxe"
   channelNames {a="", b="", g=Foward, r=""}
   defaultValue=<rgba 0 0 0.5 0>
  }
  n174 = VectorParameter() {
   group=Effect_HotWax_Warp
   parameterName=EHWax_Warp_Speed_Vehicle
   channelNames {a="", b="", g="", r=Foward}
   defaultValue=<rgba 0.2 0.2 0.2 0>
  }
  n175 = VectorParameter() {
   group=Effect_HotWax_Warp
   parameterName=EHWax_Warp_Speed_Vehicle_Side
   channelNames {a="", b="", g=Foward, r=""}
   defaultValue=<rgba 0.2 0.2 0.2 0>
  }
  n176 = MaterialFunctionCall(Float4=NamedRerouteUsage()) {materialFunction=&BreakOutFloat4Components}
  n177 = LinearInterpolate(A=@n174, B=@n175, Alpha=@n176.G)
  n178 = VectorParameter() {
   group=Effect_HotWax_Warp
   parameterName="EHWax_Warp_Speed_Mech\\Character\\Back"
   channelNames {a=Foward, b="", g="", r=""}
   defaultValue=<rgba 0.2 0.2 0.2 0>
  }
  NamedRerouteDeclaration(None=LinearInterpolate(A=LinearInterpolate(A=@n173, B=@n177, Alpha=@n176.R), B=@n178, Alpha=@n176.B)) {name=UVs_Warp_Speed, nodeColor=<rgba 0.0129 0.2456 1 1>}
  NamedRerouteDeclaration(None=@n119) {name=UVs_ObjectA_Sine, nodeColor=<rgba 0.6953 0.0906 1 1>}
  n179 = VectorParameter() {group=Effect_HotWax_Gradent_Uv, parameterName=EHWax_Grad_Offset_Vehicle, defaultValue=<rgba 0 0 0 1>}
  n180 = VectorParameter() {group=Effect_HotWax_Gradent_Uv, parameterName=EHWax_Grad_Offset_VehicleSide, defaultValue=<rgba 0 0 0 1>}
  n181 = MaterialFunctionCall(Float4=NamedRerouteUsage()) {materialFunction=&BreakOutFloat4Components}
  n182 = LinearInterpolate(
   A=LinearInterpolate(
    A=VectorParameter() {group=Effect_HotWax_Gradent_Uv, parameterName=EHWax_Grad_Offset_Weapon, defaultValue=<rgba -0.5 0 0 0>}.RGB
    B=LinearInterpolate(A=@n179.RGB, B=@n180.RGB, Alpha=@n181.G)
    Alpha=@n181.R
   )
   B=LinearInterpolate(
    A=VectorParameter() {group=Effect_HotWax_Gradent_Uv, parameterName=EHWax_Grad_Offset_Mech, defaultValue=<rgba 0 0 0 1>}.RGB
    B=VectorParameter() {group=Effect_HotWax_Gradent_Uv, parameterName=EHWax_Grad_Offset_Character, defaultValue=<rgba 0 0 0 1>}.RGB
    Alpha=@n181.A
   )
   Alpha=@n181.B
  )
  n183 = VectorParameter() {group=Effect_HotWax_Gradent_Uv, parameterName=EHWax_Grad_Offset_Backpack, defaultValue=<rgba 0 0 0 1>}
  n184 = MaterialFunctionCall(Float2=NamedRerouteUsage()) {materialFunction=&BreakOutFloat2Components}
  NamedRerouteDeclaration(None=LinearInterpolate(
    A=LinearInterpolate(A=@n182, B=@n183.RGB, Alpha=@n184.R)
    B=VectorParameter() {group=Effect_HotWax_Gradent_Uv, parameterName=EHWax_Grad_Offset_Pickaxe, defaultValue=<rgba 0 0 0 1>}.RGB
    Alpha=@n184.G
   )) {name=UVs_grad_Offset, nodeColor=<rgba 0 0.9125 1 1>}
  n185 = ScalarParameter() {defaultValue=5, group=Effect_HotWax_Object_A, parameterName=EHWax_ObjectA_TileRate_Vehicle}
  n186 = ScalarParameter() {defaultValue=5, group=Effect_HotWax_Object_A, parameterName=EHWax_ObjectA_TileRate_SideVehicle}
  n187 = MaterialFunctionCall(Float4=NamedRerouteUsage()) {materialFunction=&BreakOutFloat4Components}
  n188 = LinearInterpolate(
   A=LinearInterpolate(
    A=ScalarParameter() {defaultValue=5, group=Effect_HotWax_Object_A, parameterName=EHWax_ObjectA_TileRate_Weapon}
    B=LinearInterpolate(A=@n185, B=@n186, Alpha=@n187.G)
    Alpha=@n187.R
   )
   B=LinearInterpolate(
    A=ScalarParameter() {defaultValue=5, group=Effect_HotWax_Object_A, parameterName=EHWax_ObjectA_TileRate_Mech}
    B=ScalarParameter() {defaultValue=5, group=Effect_HotWax_Object_A, parameterName=EHWax_ObjectA_TileRate_Character}
    Alpha=@n187.A
   )
   Alpha=@n187.B
  )
  n189 = ScalarParameter() {defaultValue=5, group=Effect_HotWax_Object_A, parameterName=EHWax_ObjectA_TileRate_Backpack}
  n190 = MaterialFunctionCall(Float2=NamedRerouteUsage()) {materialFunction=&BreakOutFloat2Components}
  NamedRerouteDeclaration(None=LinearInterpolate(
    A=LinearInterpolate(A=@n188, B=@n189, Alpha=@n190.R)
    B=ScalarParameter() {defaultValue=5, group=Effect_HotWax_Object_A, parameterName=EHWax_ObjectA_TileRate_Pickaxe}
    Alpha=@n190.G
   )) {name=UVs_ObjectA_TileRate, nodeColor=<rgba 0.6953 0.0906 1 1>}
  n191 = ScalarParameter() {defaultValue=-0.25, group=Effect_HotWax_Object_A, parameterName="EHWax_ObjectA_Rotation_Weapon\\Pickaxe", sortPriority=150}
  n192 = MaterialFunctionCall(Float4=NamedRerouteUsage()) {materialFunction=&BreakOutFloat4Components}
  n193 = LinearInterpolate(
   A=ScalarParameter() {group=Effect_HotWax_Object_A, parameterName=EHWax_ObjectA_Rotation_Vehicle, sortPriority=150}
   B=ScalarParameter() {group=Effect_HotWax_Object_A, parameterName=EHWax_ObjectA_Rotation_Vehicle_Side, sortPriority=150}
   Alpha=@n192.G
  )
  n194 = ScalarParameter() {defaultValue=0.25, group=Effect_HotWax_Object_A, parameterName="EHWax_ObjectA_Rotation_Mech\\Character\\Back", sortPriority=150}
  NamedRerouteDeclaration(None=LinearInterpolate(A=LinearInterpolate(A=@n191, B=@n193, Alpha=@n192.R), B=@n194, Alpha=@n192.B)) {name=UVs_ObjectA_Rotation, nodeColor=<rgba 0.6953 0.0906 1 1>}
  n195 = StaticBoolParameter() {group=Effect_HotWax_Object, parameterName=EHWax_ObjectS_Sine}
  NamedRerouteDeclaration(None=@n195) {name=ObjectA_SineState, nodeColor=<rgba 0 0.3672 1 1>}
  NamedRerouteDeclaration(None=StaticBoolParameter() {group=Effect_HotWax_Object, parameterName=EHWax_ObjectS_Tile_Texture, sortPriority=1}) {name=ObjectA_TextureState, nodeColor=<rgba 0 0.3672 1 1>}
  n196 = VectorParameter() {group=Effect_HotWax_Object_A, parameterName=EHWax_ObjectA_TileRate__Weapon, defaultValue=<rgba 1 1 1 1>}
  n197 = VectorParameter() {group=Effect_HotWax_Object_A, parameterName=EHWax_ObjectA_TileRate__Vehicle, defaultValue=<rgba 1 1 1 1>}
  n198 = Multiply(A=@n197.RGB, B=@n197.RGB)
  n199 = VectorParameter() {group=Effect_HotWax_Object_A, parameterName=EHWax_ObjectA_TileRate__VehicleSide, defaultValue=<rgba 1 1 1 1>}
  n200 = Multiply(A=@n199.RGB, B=@n199.RGB)
  n201 = MaterialFunctionCall(Float4=NamedRerouteUsage()) {materialFunction=&BreakOutFloat4Components}
  n202 = VectorParameter() {group=Effect_HotWax_Object_A, parameterName=EHWax_ObjectA_TileRate__Mech, defaultValue=<rgba 1 1 1 1>}
  n203 = VectorParameter() {group=Effect_HotWax_Object_A, parameterName=EHWax_ObjectA_TileRate__Character, defaultValue=<rgba 1 1 1 1>}
  n204 = LinearInterpolate(
   A=LinearInterpolate(A=Multiply(A=@n196.RGB, B=@n196.RGB), B=LinearInterpolate(A=@n198, B=@n200, Alpha=@n201.G), Alpha=@n201.R)
   B=LinearInterpolate(A=Multiply(A=@n202.RGB, B=@n202.RGB), B=Multiply(A=@n203.RGB, B=@n203.RGB), Alpha=@n201.A)
   Alpha=@n201.B
  )
  n205 = VectorParameter() {group=Effect_HotWax_Object_A, parameterName=EHWax_ObjectA_TileRate__Backpack, defaultValue=<rgba 1 1 1 1>}
  n206 = Multiply(A=@n205.RGB, B=@n205.RGB)
  n207 = MaterialFunctionCall(Float2=NamedRerouteUsage()) {materialFunction=&BreakOutFloat2Components}
  n208 = VectorParameter() {group=Effect_HotWax_Object_A, parameterName=EHWax_ObjectA_TileRate__Pickaxe, defaultValue=<rgba 1 1 1 1>}
  NamedRerouteDeclaration(None=LinearInterpolate(A=LinearInterpolate(A=@n204, B=@n206, Alpha=@n207.R), B=Multiply(A=@n208.RGB, B=@n208.RGB), Alpha=@n207.G)) {name=UVs_ObjectA_TileRate_, nodeColor=<rgba 0.6953 0.0906 1 1>}
  n209 = NamedRerouteUsage()
  n210 = MaterialFunctionCall(
   Caculated_Bounds=@n209
   Pre_Skinned_Normals=NamedRerouteUsage()
   is_State=NamedRerouteUsage()
   UV_Rotation=NamedRerouteUsage()
   No_Rotation=StaticBoolParameter() {group=Effect_HotWax_Object, parameterName=EHWax_ObjectA_Rotation_Off}
  ) {materialFunction=&MF_Projection_Side_Normal_V2}
  n211 = NamedRerouteDeclaration(None=@n210.UVs) {name=Objects_A_Uvs, nodeColor=<rgba 0.9688 0 1 1>}
  n212 = ComponentMask(None=Add(A=@n2.RGB, B=NamedRerouteUsage())) {g=true, r=true}
  n213 = Multiply(A=VectorParameter() {group=Effect_HotWax_Object_A, parameterName=EHWax_ObjectA_Speed_}.RGB, B=Time())
  n214 = Add(A=Add(A=Multiply(A=@n211, B=@n212), B=ComponentMask(None=@n213) {g=true, r=true}), B=@n1)
  NamedRerouteDeclaration(None=ChannelMaskParameter(None=TextureSample(UVs=@n214, Tex=@n25).RGBA) {group=Effect_HotWax_Object, parameterName=EHWax_ObjectA_Channel_Tex_Msk}.R) {name=ObjectA_Tile, nodeColor=<rgba 0 1 0.6719 1>}
  NamedRerouteDeclaration(None=@n105.RGBA) {name=CM_Mask, nodeColor=<rgba 1 0 0.3047 1>}
  NamedRerouteDeclaration(None=@n210.UVs No Offsets/Tiling) {name=Uvs_Raw, nodeColor=<rgba 0 0.7187 1 1>}
  n215 = StaticSwitchParameter(True=NamedRerouteUsage(), False=NamedRerouteUsage()) {group=Effect_HotWax_Gradent, parameterName="EHWax_Grad  UseFull Weapon Bounds", sortPriority=1}
  n216 = NamedRerouteUsage()
  n217 = NamedRerouteUsage()
  n218 = NamedRerouteUsage()
  n219 = NamedRerouteUsage()
  n220 = NamedRerouteUsage()
  n221 = Constant3Vector() {constant=<rgba 1 0 0 0>}
  n222 = StaticSwitchParameter(True=Add(A=@n219, B=Multiply(A=@n220, B=@n221)), False=@n219) {group=Effect_HotWax_Gradent_Reactive, parameterName=EHWax_Grad----Offset Reactive}
  n223 = StaticBoolParameter() {defaultValue=true, group=Effect_HotWax_Gradent_Uv, parameterName=EHWax_Grad_Rotation_Off, sortPriority=0}
  n224 = MaterialFunctionCall(
   Caculated_Bounds=@n215
   Pre_Skinned_Normals=@n216
   is_State=@n217
   UV_Rotation=@n218
   Uvs_TileRate=StaticSwitch(True=Constant() {r=1}, False=NamedRerouteUsage(), Value=@n125)
   Uvs_Offset_Pivot=@n222
   No_Rotation=@n223
  ) {materialFunction=&MF_Projection_Side_Normal_V2}
  n225 = Multiply(
   A=NamedRerouteUsage()
   B=ScalarParameter() {defaultValue=0.05, group=Effect_HotWax, parameterName=EHWax_Grad_Warp_Strength, sortPriority=150}
  )
  n226 = Add(A=@n224.UVs, B=@n225)
  n227 = ScalarParameter() {group=Effect_HotWax_Gradent_Uv, parameterName="EHWax_Grad_Clamp_UVs  1=On", sortPriority=1}
  n228 = LinearInterpolate(A=@n226, B=Saturate(None=@n226), Alpha=@n227)
  n229 = Constant2Vector() {r=-0.5}
  n230 = NamedRerouteUsage()
  n231 = Abs(None=Multiply(A=Add(A=@n228, B=@n229) {constB=-0.5}, B=ComponentMask(None=@n230) {g=true, r=true}))
  n232 = StaticSwitch(True=LinearInterpolate(A=@n231, B=Saturate(None=@n231), Alpha=@n227), False=@n228, Value=NamedRerouteUsage())
  n233 = StaticSwitchParameter(True=NamedRerouteUsage(), False=NamedRerouteUsage()) {desc=Look To TopSide Shared, group=Effect_HotWax_Gradent, parameterName="EHWax__MSk_Tex_Grad -->TopS_00B_Sharerd_Tex_Msk", sortPriority=5}
  n234 = StaticSwitchParameter(True=NamedRerouteUsage(), False=@n233) {desc=Look To TopSide Shared, group=Effect_HotWax_Gradent, parameterName="EHWax__MSk_Tex_Grad -->TopS_00A_Sharerd_Tex_Msk", sortPriority=5}
  n235 = ChannelMaskParameter(None=@n224.UVs) {group=Effect_HotWax_Gradent, parameterName=EHWax_Grad_Mask}
  n236 = Abs(None=Add(A=@n235.R) {constB=-0.5})
  n237 = NamedRerouteUsage()
  n238 = Multiply(A=Add(A=Add(A=@n236, B=@n237) {constB=0}, B=@n225), B=ComponentMask(None=NamedRerouteUsage()) {r=true}) {constB=2}
  n239 = StaticSwitchParameter(
   True=ChannelMaskParameter(None=TextureSample(UVs=@n232, Tex=@n234).RGBA) {group=Effect_HotWax_Gradent, parameterName=EHWax_Grad_Mask}.R
   False=StaticSwitch(True=Saturate(None=@n238), False=Add(A=@n225, B=@n235.R), Value=NamedRerouteUsage())
  ) {group=Effect_HotWax_Gradent, parameterName="EHWax_Grad--------->MSk_Tex", sortPriority=1}
  n240 = ScalarParameter() {defaultValue=4, group=Effect_HotWax_Gradent, parameterName=EHWax_Grad_Multiplyer}
  NamedRerouteDeclaration(None=Power(Base=Multiply(A=@n239, B=@n240) {constB=4}, Exp=NamedRerouteUsage()) {constExponent=1}) {name=Gradent, nodeColor=<rgba 1 0.9609 0 1>}
  NamedRerouteDeclaration(None=@n124) {name=EHWax_ObjectB_Sine, nodeColor=<rgba 1 0.0276 0.7428 1>}
  n241 = ScalarParameter() {defaultValue=6, group=Effect_HotWax_Object_B, parameterName=EHWax_ObjectB_TileRate_Vehicle}
  n242 = ScalarParameter() {defaultValue=6, group=Effect_HotWax_Object_B, parameterName=EHWax_ObjectB_TileRate_SideVehicle}
  n243 = MaterialFunctionCall(Float4=NamedRerouteUsage()) {materialFunction=&BreakOutFloat4Components}
  n244 = LinearInterpolate(
   A=LinearInterpolate(
    A=ScalarParameter() {defaultValue=6, group=Effect_HotWax_Object_B, parameterName=EHWax_ObjectB_TileRate_Weapon}
    B=LinearInterpolate(A=@n241, B=@n242, Alpha=@n243.G)
    Alpha=@n243.R
   )
   B=LinearInterpolate(
    A=ScalarParameter() {defaultValue=6, group=Effect_HotWax_Object_B, parameterName=EHWax_ObjectB_TileRate_Mech}
    B=ScalarParameter() {defaultValue=6, group=Effect_HotWax_Object_B, parameterName=EHWax_ObjectB_TileRate_Character}
    Alpha=@n243.A
   )
   Alpha=@n243.B
  )
  n245 = ScalarParameter() {defaultValue=6, group=Effect_HotWax_Object_B, parameterName=EHWax_ObjectB_TileRate_Backpack}
  n246 = MaterialFunctionCall(Float2=NamedRerouteUsage()) {materialFunction=&BreakOutFloat2Components}
  NamedRerouteDeclaration(None=LinearInterpolate(
    A=LinearInterpolate(A=@n244, B=@n245, Alpha=@n246.R)
    B=ScalarParameter() {defaultValue=6, group=Effect_HotWax_Object_B, parameterName=EHWax_ObjectB_TileRate_Pickaxe}
    Alpha=@n246.G
   )) {name=EHWax_ObjectB_TileRate, nodeColor=<rgba 1 0.0276 0.7428 1>}
  n247 = ScalarParameter() {defaultValue=-0.25, group=Effect_HotWax_Object_B, parameterName="EHWax_ObjectB_Rotation_Weapon\\Pickaxe", sortPriority=150}
  n248 = MaterialFunctionCall(Float4=NamedRerouteUsage()) {materialFunction=&BreakOutFloat4Components}
  n249 = LinearInterpolate(
   A=ScalarParameter() {group=Effect_HotWax_Object_B, parameterName=EHWax_ObjectB_Rotation_Vehicle, sortPriority=150}
   B=ScalarParameter() {group=Effect_HotWax_Object_B, parameterName=EHWax_ObjectB_Rotation_Vehicle_Side, sortPriority=150}
   Alpha=@n248.G
  )
  n250 = ScalarParameter() {defaultValue=0.25, group=Effect_HotWax_Object_B, parameterName="EHWax_ObjectB_Rotation_Mech\\Character\\Back", sortPriority=150}
  NamedRerouteDeclaration(None=LinearInterpolate(A=LinearInterpolate(A=@n247, B=@n249, Alpha=@n248.R), B=@n250, Alpha=@n248.B)) {name=EHWax_ObjectB_Rotation, nodeColor=<rgba 1 0.0276 0.7428 1>}
  n251 = VectorParameter() {group=Effect_HotWax_Object_B, parameterName=EHWax_ObjectB_TileRate__Weapon, defaultValue=<rgba 1 1 1 1>}
  n252 = VectorParameter() {group=Effect_HotWax_Object_B, parameterName=EHWax_ObjectB_TileRate__Vehicle, defaultValue=<rgba 1 1 1 1>}
  n253 = Multiply(A=@n252.RGB, B=@n252.RGB)
  n254 = VectorParameter() {group=Effect_HotWax_Object_B, parameterName=EHWax_ObjectB_TileRate__VehicleSide, defaultValue=<rgba 1 1 1 1>}
  n255 = Multiply(A=@n254.RGB, B=@n254.RGB)
  n256 = MaterialFunctionCall(Float4=NamedRerouteUsage()) {materialFunction=&BreakOutFloat4Components}
  n257 = VectorParameter() {group=Effect_HotWax_Object_B, parameterName=EHWax_ObjectB_TileRate__Mech, defaultValue=<rgba 1 1 1 1>}
  n258 = VectorParameter() {group=Effect_HotWax_Object_B, parameterName=EHWax_ObjectB_TileRate__Character, defaultValue=<rgba 1 1 1 1>}
  n259 = LinearInterpolate(
   A=LinearInterpolate(A=Multiply(A=@n251.RGB, B=@n251.RGB), B=LinearInterpolate(A=@n253, B=@n255, Alpha=@n256.G), Alpha=@n256.R)
   B=LinearInterpolate(A=Multiply(A=@n257.RGB, B=@n257.RGB), B=Multiply(A=@n258.RGB, B=@n258.RGB), Alpha=@n256.A)
   Alpha=@n256.B
  )
  n260 = VectorParameter() {group=Effect_HotWax_Object_B, parameterName=EHWax_ObjectB_TileRate__Backpack, defaultValue=<rgba 1 1 1 1>}
  n261 = Multiply(A=@n260.RGB, B=@n260.RGB)
  n262 = MaterialFunctionCall(Float2=NamedRerouteUsage()) {materialFunction=&BreakOutFloat2Components}
  n263 = VectorParameter() {group=Effect_HotWax_Object_B, parameterName=EHWax_ObjectB_TileRate__Pickaxe, defaultValue=<rgba 1 1 1 1>}
  NamedRerouteDeclaration(None=LinearInterpolate(A=LinearInterpolate(A=@n259, B=@n261, Alpha=@n262.R), B=Multiply(A=@n263.RGB, B=@n263.RGB), Alpha=@n262.G)) {name=EHWax_ObjectB_TileRate_, nodeColor=<rgba 1 0.0276 0.7428 1>}
  NamedRerouteDeclaration(None=@n195) {name=ObjectB_SineState_1, nodeColor=<rgba 0 0.3672 1 1>}
  n264 = MaterialFunctionCall!(
   Index=ScalarParameter() {defaultValue=180, group=Effect_HotWax, parameterName="EHWax_CC_Shape  Index"}
   CCinput=@n99
   CCinput_Alpha=@n99
   UV=@n98
  ) {materialFunction=&/MF_Global_CC_Tex}
  NamedRerouteDeclaration(None=ChannelMaskParameter(None=@n264.4v) {group=Effect_HotWax, parameterName=EHWax_CC_Shape_Mask_Blend_Channel}.R) {name=Blend Mask, nodeColor=<rgba 1 0 0.7266 1>}
  n265 = NamedRerouteUsage()
  n266 = MaterialFunctionCall(
   Caculated_Bounds=@n265
   Pre_Skinned_Normals=NamedRerouteUsage()
   is_State=NamedRerouteUsage()
   UV_Rotation=NamedRerouteUsage()
   No_Rotation=StaticBoolParameter() {group=Effect_HotWax_Object, parameterName=EHWax_ObjectB_Rotation_Off}
  ) {materialFunction=&MF_Projection_Side_Normal_V2}
  n267 = NamedRerouteDeclaration(None=@n266.UVs) {name=Objects_B_Uvs, nodeColor=<rgba 0.9688 0 1 1>}
  n268 = ComponentMask(None=Add(A=NamedRerouteUsage(), B=@n39.RGB)) {g=true, r=true}
  n269 = Multiply(A=Time(), B=VectorParameter() {group=Effect_HotWax_Object_B, parameterName=EHWax_ObjectB_Speed_}.RGB)
  n270 = Add(A=Add(A=Multiply(A=@n267, B=@n268), B=ComponentMask(None=@n269) {g=true, r=true}), B=@n38)
  NamedRerouteDeclaration(None=ChannelMaskParameter(None=TextureSample(UVs=@n270, Tex=@n61).RGBA) {group=Effect_HotWax_Object, parameterName=EHWax_ObjectB_Channel_Tex_Msk}.R) {name=ObjectB_Tile, nodeColor=<rgba 0 1 0.6719 1>}
  NamedRerouteDeclaration(None=FunctionInput(Preview=Constant4Vector().RGBA) {bUsePreviewValueAsDefault=true, inputName=IsState, inputType=FunctionInput_Vector4, sortPriority=4}) {name=IsState, nodeColor=<rgba 0.1497 1 0.1633 1>}
  NamedRerouteDeclaration(None=@n266.UVs No Offsets/Tiling) {name=Uvs_Raw_1, nodeColor=<rgba 0 0.7188 1 1>}
  n271 = GetMaterialAttributes(None=FunctionInput(Preview=@n111) {inputName=Material_Root, inputType=FunctionInput_MaterialAttributes}) {attributeGetTypes [
    #2
    #3
    #1
    #4
   ]}
  NamedRerouteDeclaration(None=@n271.Normal) {name=Normal_Root, nodeColor=<rgba 0.4062 0 1 1>}
  NamedRerouteDeclaration(None=@n271.MaterialAttributes) {name=Material_Root, nodeColor=<rgba 0.7812 1 0 1>}
  n272 = TextureObjectParameter() {
   group=TopSide_V02
   parameterName=TopS_00A_Sharerd_Tex_Msk
   samplerType=SAMPLERTYPE_Masks
   sortPriority=249
   texture=&T_BlackMask
  }
  NamedRerouteDeclaration(None=@n272) {name=Tex_00A_Shared_Tex, nodeColor=<rgba 1 0.8906 0 1>}
  n273 = GetMaterialAttributes(None=NamedRerouteUsage()) {attributeGetTypes [#2, #5]}
  n274 = VectorParameter() {group=Effect_HotWax, parameterName=EHWax_B_BaseColor, defaultValue=<rgba 0 0 0 1>}
  n275 = ScalarParameter() {defaultValue=1, group=Effect_HotWax, parameterName="EHWax_B_BaseColor_Darken_Root   1=On"}
  n276 = VectorParameter() {group=Effect_HotWax, parameterName=EHWax_B_Curvature_Brightness, channelNames {a="", b="", g=Pits, r=Peaks}, defaultValue=<rgba 1 1 0 1>}
  n277 = VectorParameter() {group=Effect_HotWax, parameterName=EHWax_CC SolidColor, defaultValue=<rgba 1 1 1 1>}
  n278 = ChannelMaskParameter(None=@n264.4v) {group=Effect_HotWax, maskChannel=Alpha, parameterName=EHWax_CC_Shape_Channel, defaultValue=<rgba 0 0 0 1>}
  n279 = StaticSwitchParameter(
   True=StaticSwitchParameter(True=Add(A=@n278.R, B=@n96), False=@n278.R) {group=Effect_HotWax_Noise, parameterName=EHWax_Add Noise_AfterShape, sortPriority=0}
   False=@n278.R
  ) {group=Effect_HotWax_Noise, parameterName=EHWax_Add Noise, sortPriority=0}
  n280 = CurveAtlasRowParameter(CurveTime=@n279) {
   atlas=&CruveAtlas_Wraps_01
   curve=/Game/Weapons/WeaponSkins/Wraps/Textures/LavaLamp/CC_Liquid_HotWax
   defaultValue=181
   desc=Alpah is useed as Brightness Mask
   group=Effect_HotWax
   parameterName=EHWax_CC
  }
  n281 = MaterialFunctionCall(
   Index=ScalarParameter() {defaultValue=181, group=Effect_HotWax, parameterName="EHWax_CC  Index"}
   CCinput=@n280
   CCinput_Alpha=@n280
   UV=@n279
  ) {materialFunction=&MF_Global_CC_Tex}
  n282 = StaticSwitchParameter(True=@n277.RGB, False=@n281.3V) {group=Effect_HotWax, parameterName="EHWax_CC ---->SolidColor"}
  n283 = Multiply(A=NamedRerouteUsage(), B=@n95)
  n284 = StaticSwitchParameter(
   True=LinearInterpolate(
    A=@n282
    B=VectorParameter() {group=Effect_HotWax_Noise, parameterName=EHWax_Noise_Color, defaultValue=<rgba 1 1 1 1>}.RGB
    Alpha=Saturate(None=@n283)
   )
   False=@n282
  ) {group=Effect_HotWax_Noise, parameterName=EHWax_Add Noise, sortPriority=0}
  n285 = Add(A=NamedRerouteUsage()) {constB=-0.5}
  n286 = LinearInterpolate(
   A=LinearInterpolate(A=@n284, B=Multiply(A=@n276.Peaks, B=@n284), Alpha=Saturate(None=@n285))
   B=Multiply(A=@n276.Pits, B=@n284) {constA=1}
   Alpha=Saturate(None=Multiply(A=@n285) {constB=-1})
  )
  n287 = ChannelMaskParameter(None=@n264.4v) {group=Effect_HotWax, maskChannel=Blue, parameterName=EHWax_CC_Shape_Mask_Channel, defaultValue=<rgba 0 0 1 0>}
  n288 = VectorParameter() {group=Effect_HotWax, parameterName=EHWax_GmapB_Color, defaultValue=<rgba 1 1 1 1>}
  n289 = VectorParameter() {group=Effect_HotWax, parameterName=EHWax_GmapA_Color, defaultValue=<rgba 1 1 1 1>}
  n290 = NamedRerouteUsage()
  n291 = Add(
   A=LinearInterpolate(A=@n273.EmissiveColor, Alpha=ScalarParameter() {defaultValue=1, group=Effect_HotWax, parameterName="EHWax_B_Emissive_Darken_Root   1=On"}) {constB=0}
   B=Multiply(A=Multiply(A=@n286, B=@n287.R), B=LinearInterpolate(A=@n288.RGB, B=@n289.RGB, Alpha=@n290))
  )
  n292 = ScalarParameter() {group=Effect_HotWax, parameterName="EHWax_B-Emissive---->BaseColor"}
  n293 = ChannelMaskParameter(None=@n264.4v) {group=Effect_HotWax, maskChannel=Green, parameterName=EHWax_CC_Shape_Roughness_Channel, defaultValue=<rgba 0 1 0 0>}
  n294 = NamedRerouteUsage()
  n295 = Multiply(A=@n291, B=StaticSwitchParameter(True=@n277.A, False=@n281.Alpha) {group=Effect_HotWax, parameterName="EHWax_CC ---->SolidColor"})
  n296 = VectorParameter() {desc="CC_Alpha as Muliplyer mask ", group=Effect_HotWax, parameterName=EHWax_B_Emissive Color_Multiplier, defaultValue=<rgba 0 0 0 1>}
  n297 = CollectionParameter() {
   collection=&FortniteMaterialParameters
   parameterName=SunlightDirectionalVector
  }
  n298 = ComponentMask(None=@n297) {b=true, g=true, r=true}
  n299 = MaterialFunctionCall!(Value=@n279) {materialFunction=/Engine/Functions/Engine_MaterialFunctions03/Procedurals/PreparePerturbNormalHQ}
  n300 = MaterialFunctionCall!(
   Bump one pixel right=@n299.Value one pixel right
   Bump center=@n299.Value one pixel right
   Bump one pixel down=@n299.Value one pixel right
   WS Normal
  ) {materialFunction=/Engine/Functions/Engine_MaterialFunctions03/Procedurals/PerturbNormalHQ}
  n301 = Normalize(VectorInput=@n300.WS Normal)
  n302 = Multiply(A=DotProduct(A=@n298, B=@n301)) {constB=2}
  n303 = Normalize(VectorInput=Subtract(A=Multiply(A=@n302, B=@n301), B=@n298))
  n304 = Multiply(A=MaterialFunctionCall() {materialFunction=&CameraDirectionVector}.Result) {constB=-1}
  n305 = Multiply(
   A=Power(Base=DotProduct(A=@n303, B=@n304), Exp=ScalarParameter() {defaultValue=5, group=Effect_HotWax, parameterName=EHWax_Fake_lighting_Falloff})
   B=ScalarParameter() {defaultValue=0.4, group=Effect_HotWax, parameterName=EHWax_Fake_lighting_Strength}
  ) {constB=0.4}
  n306 = LinearInterpolate(
   A=Add(A=LinearInterpolate(A=@n291, Alpha=@n292) {constB=0}, B=Multiply(A=@n295, B=@n296.RGB))
   B=VectorParameter() {group=Effect_HotWax, parameterName=EHWax_Fake_lighting_Color, defaultValue=<rgba 1 1 1 1>}.RGB
   Alpha=Multiply(A=@n287.R, B=Saturate(None=@n305))
  )
  n307 = LinearInterpolate(
   A=NamedRerouteUsage()
   B=NamedRerouteUsage()
   Alpha=ScalarParameter() {defaultValue=1, group=Effect_HotWax, parameterName="EHWax_Normal_Root 1=On"}
  )
  n308 = Multiply(
   A=Transform(None=@n300.WS Normal) {transformSourceType=TRANSFORMSOURCE_World, transformType=TRANSFORM_Tangent}
   B=Constant3Vector() {constant=<rgba 1 1 0 0>}
  ) {constB=3}
  n309 = SetMaterialAttributes!(
   MaterialAttributes=NamedRerouteUsage()
   BaseColor=LinearInterpolate(A=LinearInterpolate(A=@n273.BaseColor, B=@n274.RGB, Alpha=@n275) {constB=0}, B=@n291, Alpha=@n292)
   Metallic=ScalarParameter() {group=Effect_HotWax, parameterName=EHWax_B_Metalic}
   Specular=ScalarParameter() {defaultValue=0.5, group=Effect_HotWax, parameterName=EHWax_B_Spec}
   Roughness=LinearInterpolate(
    A=ScalarParameter() {defaultValue=1, group=Effect_HotWax, parameterName=EHWax_B_Roughness_Low}
    B=ScalarParameter() {defaultValue=0.2, group=Effect_HotWax, parameterName=EHWax_B_Roughness_High}
    Alpha=@n293.R
   ) {constA=1, constB=0}
   EmissiveColor=StaticSwitchParameter(True=Multiply(A=@n294, B=@n306), False=@n306) {group="Effect_HotWax   Reactive", parameterName="EHWax----Brightness  Reactive"}
   Normal=LinearInterpolate(A=@n307, B=Add(A=@n307, B=@n308), Alpha=ScalarParameter() {defaultValue=1, group=Effect_HotWax, parameterName=EHWax_Normal_Strength})
  ) {attributeSetTypes [
    #2
    #4
    #6
    #3
    #5
    #1
   ]}
  NamedRerouteDeclaration(None=@n309) {name=Material, nodeColor=<rgba 0 1 0.9531 1>}
  n310 = MaterialFunctionCall(Color=Abs(None=NamedRerouteUsage())) {materialFunction=&SplitComponents}
  n311 = MaterialFunctionCall(Float4=NamedRerouteUsage()) {materialFunction=&BreakOutFloat4Components}
  n312 = LinearInterpolate(A=@n310.R, B=LinearInterpolate(A=@n310.B, B=@n310.B, Alpha=@n311.G), Alpha=@n311.R)
  NamedRerouteDeclaration(None=MaterialFunctionCall(
    In=LinearInterpolate(A=@n312, B=@n310.G, Alpha=@n311.B)
    Contrast=ScalarParameter() {defaultValue=1, group=Effect_HotWax, parameterName=EHWax_Tri_Blend_Contrast}
   ) {materialFunction=&CheapContrast}.Result) {name=Tri_Mask, nodeColor=<rgba 1 0.4688 0 1>}
  n313 = NamedRerouteUsage()
  n314 = NamedRerouteUsage()
  n315 = NamedRerouteUsage()
  n316 = Time()
  n317 = VectorParameter() {group=Effect_HotWax_Warp, parameterName=EHWax_Warp_Speed, channelNames {a="", b="", g="", r=Foward}, defaultValue=<rgba 0.2 0.2 0.2 0>}
  n318 = Multiply(A=@n316, B=@n317)
  n319 = NamedRerouteUsage()
  n320 = StaticBool() {value=true}
  n321 = StaticSwitchParameter(True=NamedRerouteUsage(), False=NamedRerouteUsage()) {desc=Look To TopSide Shared, group=Effect_HotWax_Warp, parameterName="EHWax__MSk_Tex_Warp-->TopS_00B_Sharerd_Tex_Msk", sortPriority=5}
  n322 = StaticSwitchParameter(True=NamedRerouteUsage(), False=@n321) {desc=Look To TopSide Shared, group=Effect_HotWax_Warp, parameterName="EHWax__MSk_Tex_Warp -->TopS_00A_Sharerd_Tex_Msk", sortPriority=5}
  n323 = ChannelMaskParameter(None=TextureSample(
    UVs=MaterialFunctionCall(Caculated_Bounds=@n313, Pre_Skinned_Normals=@n314, is_State=@n315, Uvs_Offset=@n318, Uvs_TileRate=@n319, No_Rotation=@n320) {materialFunction=&MF_Projection_Side_Normal_V2}.UVs
    Tex=@n322
   ).RGBA) {group=Effect_HotWax_Warp, maskChannel=Alpha, parameterName=EHWax_Warp_Channel_Tex_Msk, sortPriority=1, defaultValue=<rgba 0 0 0 1>}
  n324 = Add(A=@n323.R) {constB=-0.5}
  n325 = Add(A=Multiply(A=NamedRerouteUsage(), B=@n319), B=Multiply(A=@n316, B=NamedRerouteUsage()))
  n326 = ChannelMaskParameter(None=Multiply(A=VectorNoise(World Position=@n325) {noiseFunction=VNF_VectorALU}) {constB=0.5}) {group=Effect_HotWax_Warp, parameterName=EHWax_Warp_Noise_Channel, sortPriority=1}
  n327 = Add(A=NamedRerouteUsage()) {constB=-0.5}
  n328 = ScalarParameter() {group=Effect_HotWax_Gradent_Uv, parameterName=EHWax_Curavature_Warp_Strength, sortPriority=1}
  NamedRerouteDeclaration(None=Add(
    A=StaticSwitchParameter(True=@n324, False=@n326.R) {group=Effect_HotWax_Warp, parameterName="EHWax_Warp_procedural --------->MSk_Tex", sortPriority=0}
    B=Multiply(A=@n327, B=@n328)
   )) {name=Warp_Object, nodeColor=<rgba 0 0.3906 1 1>}
  n329 = VectorParameter() {group=Effect_HotWax_Warp, parameterName=EHWax_Object_WarpScale_Weapon, defaultValue=<rgba 1 1 1 10>}
  n330 = VectorParameter() {group=Effect_HotWax_Warp, parameterName=EHWax_Object_WarpScale_Vehicle, defaultValue=<rgba 1 1 1 10>}
  n331 = Multiply(A=@n330.RGB, B=@n330.RGB)
  n332 = VectorParameter() {group=Effect_HotWax_Warp, parameterName=EHWax_Object_WarpScale_VehicleSide, defaultValue=<rgba 1 1 1 10>}
  n333 = Multiply(A=@n332.RGB, B=@n332.RGB)
  n334 = MaterialFunctionCall(Float4=NamedRerouteUsage()) {materialFunction=&BreakOutFloat4Components}
  n335 = VectorParameter() {group=Effect_HotWax_Warp, parameterName=EHWax_Object_WarpScale_Mech, defaultValue=<rgba 1 1 1 10>}
  n336 = VectorParameter() {group=Effect_HotWax_Warp, parameterName=EHWax_Object_WarpScale_Character, defaultValue=<rgba 1 1 1 10>}
  n337 = LinearInterpolate(
   A=LinearInterpolate(A=Multiply(A=@n329.RGB, B=@n329.RGB), B=LinearInterpolate(A=@n331, B=@n333, Alpha=@n334.G), Alpha=@n334.R)
   B=LinearInterpolate(A=Multiply(A=@n335.RGB, B=@n335.RGB), B=Multiply(A=@n336.RGB, B=@n336.RGB), Alpha=@n334.A)
   Alpha=@n334.B
  )
  n338 = VectorParameter() {group=Effect_HotWax_Warp, parameterName=EHWax_Object_WarpScale_Backpack, defaultValue=<rgba 1 1 1 10>}
  n339 = Multiply(A=@n338.RGB, B=@n338.RGB)
  n340 = MaterialFunctionCall(Float2=NamedRerouteUsage()) {materialFunction=&BreakOutFloat2Components}
  n341 = VectorParameter() {group=Effect_HotWax_Warp, parameterName=EHWax_Object_WarpScale_Pickaxe, defaultValue=<rgba 1 1 1 10>}
  NamedRerouteDeclaration(None=LinearInterpolate(A=LinearInterpolate(A=@n337, B=@n339, Alpha=@n340.R), B=Multiply(A=@n341.RGB, B=@n341.RGB), Alpha=@n340.G)) {name=UVs_Object_Warp_Scale_1, nodeColor=<rgba 0 0.1613 1 1>}
  NamedRerouteDeclaration(None=@n106.A) {name=Msk_Gmap, nodeColor=<rgba 0 0.4844 1 1>}
  n342 = VectorParameter() {group=Effect_HotWax_Noise, parameterName=EHWax_NoiseScale_Weapon, sortPriority=150, defaultValue=<rgba 1 1 1 4>}
  n343 = VectorParameter() {group=Effect_HotWax_Noise, parameterName=EHWax_NoiseScale_Vehicle, sortPriority=150, defaultValue=<rgba 1 1 1 4>}
  n344 = Multiply(A=@n343.RGB, B=@n343.RGB)
  n345 = VectorParameter() {group=Effect_HotWax_Noise, parameterName=EHWax_NoiseScale_VehicleSide, sortPriority=150, defaultValue=<rgba 1 1 1 4>}
  n346 = Multiply(A=@n345.RGB, B=@n345.RGB)
  n347 = MaterialFunctionCall(Float4=NamedRerouteUsage()) {materialFunction=&BreakOutFloat4Components}
  n348 = VectorParameter() {group=Effect_HotWax_Noise, parameterName=EHWax_NoiseScale_Mech, sortPriority=150, defaultValue=<rgba 1 1 1 4>}
  n349 = VectorParameter() {group=Effect_HotWax_Noise, parameterName=EHWax_NoiseScale_Character, sortPriority=150, defaultValue=<rgba 1 1 1 4>}
  n350 = LinearInterpolate(
   A=LinearInterpolate(A=Multiply(A=@n342.RGB, B=@n342.RGB), B=LinearInterpolate(A=@n344, B=@n346, Alpha=@n347.G), Alpha=@n347.R)
   B=LinearInterpolate(A=Multiply(A=@n348.RGB, B=@n348.RGB), B=Multiply(A=@n349.RGB, B=@n349.RGB), Alpha=@n347.A)
   Alpha=@n347.B
  )
  n351 = VectorParameter() {group=Effect_HotWax_Noise, parameterName=EHWax_NoiseScale_Backpack, sortPriority=150, defaultValue=<rgba 1 1 1 4>}
  n352 = Multiply(A=@n351.RGB, B=@n351.RGB)
  n353 = MaterialFunctionCall(Float2=NamedRerouteUsage()) {materialFunction=&BreakOutFloat2Components}
  n354 = VectorParameter() {group=Effect_HotWax_Noise, parameterName=EHWax_NoiseScale_Pickaxe, sortPriority=150, defaultValue=<rgba 1 1 1 4>}
  NamedRerouteDeclaration(None=LinearInterpolate(A=LinearInterpolate(A=@n350, B=@n352, Alpha=@n353.R), B=Multiply(A=@n354.RGB, B=@n354.RGB), Alpha=@n353.G)) {name=UVs_Noise_Scale, nodeColor=<rgba 0.3229 1 0.4004 1>}
  NamedRerouteDeclaration(None=FunctionInput(Preview=Constant4Vector().RGBA) {bUsePreviewValueAsDefault=true, inputName=IsState_B, inputType=FunctionInput_Vector4, sortPriority=4}) {name=IsState_B, nodeColor=<rgba 0.1497 1 0.1633 1>}
  n355 = VectorParameter() {group=Effect_HotWax_Noise, parameterName=EHWax_Noise_Speed, channelNames {a="", b="", g="", r=Foward}, defaultValue=<rgba 0.2 0.2 0.2 0>}
  n356 = MaterialFunctionCall(
   Caculated_Bounds=NamedRerouteUsage()
   Pre_Skinned_Normals=NamedRerouteUsage()
   is_State=NamedRerouteUsage()
   UV_Rotation=NamedRerouteUsage()
   Uvs_Offset=Multiply(A=Time(), B=@n355)
   Uvs_TileRate=NamedRerouteUsage()
   No_Rotation=StaticBoolParameter() {defaultValue=true, group=Effect_HotWax_Noise, parameterName=EHWax_Noise_Rotation, sortPriority=150}
  ) {materialFunction=&MF_Projection_Side_Normal_V2}
  n357 = Multiply(A=NamedRerouteUsage(), B=ScalarParameter() {group=Effect_HotWax_Noise, parameterName=EHWax_Noise__Warp_Amount})
  n358 = StaticSwitchParameter(True=NamedRerouteUsage(), False=NamedRerouteUsage()) {desc=Look To TopSide Shared, group=Effect_HotWax_Noise, parameterName="EHWax__MSk_Tex_Noise-->TopS_00B_Sharerd_Tex_Msk", sortPriority=5}
  n359 = StaticSwitchParameter(True=NamedRerouteUsage(), False=@n358) {desc=Look To TopSide Shared, group=Effect_HotWax_Noise, parameterName="EHWax__MSk_Tex_Noise -->TopS_00A_Sharerd_Tex_Msk", sortPriority=5}
  n360 = ChannelMaskParameter(None=TextureSample(UVs=Add(A=@n356.UVs, B=@n357), Tex=@n359).RGBA) {group=Effect_HotWax_Noise, maskChannel=Alpha, parameterName=EHWax_Noise_Channel_Tex_Msk, sortPriority=1, defaultValue=<rgba 0 0 0 1>}
  n361 = ScalarParameter() {defaultValue=-0.5, group=Effect_HotWax_Noise, parameterName=EHWax_Noise__Shape_Position}
  NamedRerouteDeclaration(None=Multiply(A=Add(A=@n360.R, B=@n361) {constB=-0.5}, B=ScalarParameter() {group=Effect_HotWax_Noise, parameterName=EHWax_Noise__Shape_Amount})) {name=Noise_Shape, nodeColor=<rgba 0.2891 1 0 1>}
  n362 = Add(A=@n360.R, B=ScalarParameter() {defaultValue=-0.5, group=Effect_HotWax_Noise, parameterName=EHWax_Noise__Color_Position})
  n363 = ScalarParameter() {group=Effect_HotWax_Noise, parameterName=EHWax_Noise__Color_Amount}
  NamedRerouteDeclaration(None=Power(Base=Multiply(A=@n362, B=@n363), Exp=ScalarParameter() {defaultValue=2, group=Effect_HotWax_Noise, parameterName=EHWax_Noise__Color_Falloff})) {name=Noise_Color, nodeColor=<rgba 0.2891 1 0 1>}
  n364 = ScalarParameter() {group=Effect_HotWax_Noise, parameterName="EHWax_Noise_Rotation_Weapon\\Pickaxe", sortPriority=150}
  n365 = MaterialFunctionCall(Float4=NamedRerouteUsage()) {materialFunction=&BreakOutFloat4Components}
  n366 = LinearInterpolate(
   A=ScalarParameter() {group=Effect_HotWax_Noise, parameterName=EHWax_Noise_Rotation_Vehicle, sortPriority=150}
   B=ScalarParameter() {group=Effect_HotWax_Noise, parameterName=EHWax_Noise_Rotation_Vehicle_Side, sortPriority=150}
   Alpha=@n365.G
  )
  NamedRerouteDeclaration(None=LinearInterpolate(
    A=LinearInterpolate(A=@n364, B=@n366, Alpha=@n365.R)
    B=ScalarParameter() {group=Effect_HotWax_Noise, parameterName="EHWax_Noise_Rotation_Mech\\Character\\Back", sortPriority=150}
    Alpha=@n365.B
   )) {name=UVs_Noies_Rotation, nodeColor=<rgba 0 0.9125 1 1>}
  n367 = PreSkinnedPosition()
  n368 = PreSkinnedLocalBounds()
  n369 = MaterialFunctionCall(Color=Subtract(A=@n368.Max, B=@n368.Max)) {materialFunction=&SplitComponents}
  n370 = Divide(A=Subtract(A=VertexInterpolator(VS=@n367).PS, B=@n368.Min), B=Max(A=Max(A=@n369.R, B=@n369.R), B=@n369.B))
  n371 = FunctionInput(Preview=@n370) {bUsePreviewValueAsDefault=true, inputName=Calcated_Bounds, sortPriority=11}
  n372 = MaterialFunctionCall(Float4=NamedRerouteUsage()) {materialFunction=&BreakOutFloat4Components}
  n373 = Max(A=@n372.R, B=@n372.R)
  n374 = MaterialFunctionCall(Float4=NamedRerouteUsage()) {materialFunction=&BreakOutFloat4Components}
  n375 = Max(A=Max(A=@n372.B, B=@n372.B), B=Max(A=@n374.R, B=@n374.R))
  NamedRerouteDeclaration(None=LinearInterpolate(A=@n370, B=NamedRerouteDeclaration(None=@n371) {name=Caclcated_Bounds, nodeColor=<rgba 0 0.9297 1 1>}, Alpha=Max(A=@n373, B=@n375))) {name=Full_Bounds, nodeColor=<rgba 0 0.7891 1 1>}
  NamedRerouteDeclaration(None=ChannelMaskParameter(None=MaterialFunctionCall(Caculated_Bounds=@n215, Pre_Skinned_Normals=@n216, is_State=@n217, UV_Rotation=@n218, Uvs_Offset_Pivot=@n222, No_Rotation=@n223) {materialFunction=&MF_Projection_Side_Normal_V2}.UVs) {group=Effect_HotWax_Gradent, parameterName=EHWax_Grad_Mask}.R) {name=Uvs_Gradent, nodeColor=<rgba 1 0 0.7266 1>}
  n376 = Time()
  n377 = ScalarParameter() {defaultValue=-9999999, group=Effect_HotWax_Gradent_Reactive, parameterName=EHWax_Grad---Offset_TBP----------------, sortPriority=600}
  n378 = Add(A=@n377, B=ScalarParameter() {group=Effect_HotWax_Gradent_Reactive, parameterName=EHWax_Grad---Offset_Delay})
  n379 = Saturate(None=Divide(
    A=Subtract(A=@n376, B=@n378)
    B=ScalarParameter() {defaultValue=0.25, group=Effect_HotWax_Gradent_Reactive, parameterName=EHWax_Grad---Offset_Duration}
   ))
  n380 = ScalarParameter() {defaultValue=0.25, group=Effect_HotWax_Gradent_Reactive, parameterName=EHWax_Grad---Offset_Falloff}
  NamedRerouteDeclaration(None=LinearInterpolate(
    A=ScalarParameter() {defaultValue=1, group=Effect_HotWax_Gradent_Reactive, parameterName=EHWax_Grad---Offset Start}
    B=ScalarParameter() {group=Effect_HotWax_Gradent_Reactive, parameterName=EHWax_Grad---Offset End}
    Alpha=Power(Base=@n379, Exp=@n380)
   )) {name=Grad_Reactive_Offset, nodeColor=<rgba 0.1719 0 1 1>}
  n381 = Time()
  n382 = ScalarParameter() {defaultValue=-9999999, group="Effect_HotWax   Reactive", parameterName=EHWax----Brightness--TBP----------------, sortPriority=600}
  n383 = Add(A=@n382, B=ScalarParameter() {group="Effect_HotWax   Reactive", parameterName=EHWax----Brightness_Delay})
  n384 = Saturate(None=Divide(
    A=Subtract(A=@n381, B=@n383)
    B=ScalarParameter() {defaultValue=0.25, group="Effect_HotWax   Reactive", parameterName=EHWax----Brightness_Duration}
   ))
  n385 = ScalarParameter() {defaultValue=0.25, group="Effect_HotWax   Reactive", parameterName=EHWax----Brightness_Falloff}
  NamedRerouteDeclaration(None=LinearInterpolate(
    A=ScalarParameter() {defaultValue=1, group="Effect_HotWax   Reactive", parameterName=EHWax----Brightness Start}
    B=ScalarParameter() {group="Effect_HotWax   Reactive", parameterName=EHWax----Brightness End}
    Alpha=Power(Base=@n384, Exp=@n385)
   )) {name=Grad_Reactive_Brightness, nodeColor=<rgba 0.1719 0 1 1>}
  n386 = VectorParameter() {group=Effect_HotWax_Gradent_Uv, parameterName=EHWax_Grad_Scale_Weapon, defaultValue=<rgba 1 1 1 1>}
  n387 = VectorParameter() {group=Effect_HotWax_Gradent_Uv, parameterName=EHWax_Grad_Scale_Vehicle, defaultValue=<rgba 1 1 1 1>}
  n388 = Multiply(A=@n387.RGB, B=@n387.RGB)
  n389 = VectorParameter() {group=Effect_HotWax_Gradent_Uv, parameterName=EHWax_Grad_Scale_VehicleSide, defaultValue=<rgba 1 1 1 1>}
  n390 = Multiply(A=@n389.RGB, B=@n389.RGB)
  n391 = MaterialFunctionCall(Float4=NamedRerouteUsage()) {materialFunction=&BreakOutFloat4Components}
  n392 = VectorParameter() {group=Effect_HotWax_Gradent_Uv, parameterName=EHWax_Grad_Scale_Mech, defaultValue=<rgba 1 1 1 1>}
  n393 = VectorParameter() {group=Effect_HotWax_Gradent_Uv, parameterName=EHWax_Grad_Scale_Character, defaultValue=<rgba 1 1 1 1>}
  n394 = LinearInterpolate(
   A=LinearInterpolate(A=Multiply(A=@n386.RGB, B=@n386.RGB), B=LinearInterpolate(A=@n388, B=@n390, Alpha=@n391.G), Alpha=@n391.R)
   B=LinearInterpolate(A=Multiply(A=@n392.RGB, B=@n392.RGB), B=Multiply(A=@n393.RGB, B=@n393.RGB), Alpha=@n391.A)
   Alpha=@n391.B
  )
  n395 = VectorParameter() {group=Effect_HotWax_Gradent_Uv, parameterName=EHWax_Grad_Scale_Backpack, defaultValue=<rgba 1 1 1 1>}
  n396 = Multiply(A=@n395.RGB, B=@n395.RGB)
  n397 = MaterialFunctionCall(Float2=NamedRerouteUsage()) {materialFunction=&BreakOutFloat2Components}
  n398 = VectorParameter() {group=Effect_HotWax_Gradent_Uv, parameterName=EHWax_Grad_Scale_Pickaxe, defaultValue=<rgba 1 1 1 1>}
  NamedRerouteDeclaration(None=LinearInterpolate(A=LinearInterpolate(A=@n394, B=@n396, Alpha=@n397.R), B=Multiply(A=@n398.RGB, B=@n398.RGB), Alpha=@n397.G)) {name=UVs_Grad_Scale, nodeColor=<rgba 0 0.9125 1 1>}
  n399 = PreSkinnedNormal()
  NamedRerouteDeclaration(None=FunctionInput(Preview=VertexInterpolator(VS=@n399).PS) {bUsePreviewValueAsDefault=true, inputName=Pre-Skinned Local Normal, sortPriority=10}) {name=PreSkin_Normal, nodeColor=<rgba 0.8281 1 0 1>}
  NamedRerouteDeclaration(None=@n112.Normal) {name=Root_Normal, nodeColor=<rgba 0.4062 0 1 1>}
  NamedRerouteUsage()
  NamedRerouteUsage()
  NamedRerouteUsage()
  n400 = Distance(B=Constant() {r=0.5})
  OneMinus(None=@n400)
  PixelNormalWS()
  Saturate(None=LinearInterpolate(Alpha=@n400) {constA=1, constB=-1.5})
  StaticBool() {value=true}
  StaticSwitchParameter(True=NamedRerouteUsage(), False=@n265) {group=Effect_HotWax_Object, parameterName=EHWax_ObjectS Use Full Weapon Bounds, sortPriority=0}
  StaticSwitchParameter(True=NamedRerouteUsage(), False=@n209) {group=Effect_HotWax_Object, parameterName=EHWax_ObjectS Use Full Weapon Bounds, sortPriority=0}
  VertexNormalWS()
 ]}
