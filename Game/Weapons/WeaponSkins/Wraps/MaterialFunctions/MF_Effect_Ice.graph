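# MF_Effect_Ice is a weapon-wrap material function that layers an ice effect over the incoming material behind the Effect_Ice static switch.
# Ice color combines cheap triplanar noise, gradient or curvature masks, cracks, frost and refracted skylight environment samples.
# It fakes depth and lighting with parallax inner-ice UVs, Fresnel highlights, fake SSS and GGX specular, adding time-of-day-scaled emissive.
# CM, decal and TopSide blend masks control where the ice appears.
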MaterialFunction /Game/Weapons/WeaponSkins/Wraps/MaterialFunctions/MF_Effect_Ice
graph {nodes [
  n1 = NamedRerouteUsage()
  n2 = Fresnel() {baseReflectFraction=0}
  n3 = Add(A=Add(A=NamedRerouteUsage(), B=Multiply(A=@n1) {constB=0.5}), B=Saturate(None=Multiply(A=@n2) {constB=3}))
  Add(A=@n3, B=NamedRerouteUsage())
  n4 = Divide(A=LocalPosition() {localOrigin=InstancePreSkinning}.XYZ) {constB=75}
  ComponentMask(None=NamedRerouteDeclaration(None=VertexInterpolator(VS=@n4).PS) {name=PreSkinnedPosition, nodeColor=<rgba 0 0.7891 1 1>}) {b=true, r=true}
  n5 = NamedRerouteUsage()
  n6 = FunctionInput() {inputName=Decal_Mask, inputType=FunctionInput_Scalar, sortPriority=90}
  n7 = OneMinus(None=@n6)
  n8 = ScalarParameter() {defaultValue=1, group=Effect-Ice, parameterName="EfIce_CustomMSk_Decal_Darken-->CustomMSk 0=Active", sortPriority=12}
  n9 = ChannelMaskParameter(None=FunctionInput() {inputName=Ts_Mask_02, inputType=FunctionInput_Vector4, sortPriority=20}) {group=Effect-Ice, parameterName=EfIce_CustomMSk_TS_Mask02_Channel, sortPriority=11}
  n10 = ChannelMaskParameter(None=FunctionInput() {inputName=Ts_Mask_01, inputType=FunctionInput_Vector4, sortPriority=20}) {group=Effect-Ice, parameterName=EfIce_CustomMSk_TS_Mask01_Channel, sortPriority=11}
  n11 = ChannelMaskParameter(None=FunctionInput() {inputName=Decal_RGBA, inputType=FunctionInput_Vector4, sortPriority=90}) {group=Effect-Ice, parameterName=EfIce_CustomMSk_Decal_Msk_RGBA, sortPriority=12}
  n12 = Constant()
  n13 = Add(
   A=Multiply(
    A=Max(A=@n7, B=@n8)
    B=StaticSwitchParameter(True=@n9.R, False=@n10.R) {group=Effect-Ice, parameterName="EfIce_CustomMSk_TS_Mask01-------->Mask02", sortPriority=11}
   )
   B=StaticSwitchParameter(
    True=StaticSwitchParameter(True=@n11.R, False=@n12) {group=Effect-Ice, parameterName="EfIce_CustomMSk_Add_Left\\Right-Decal_Msk", sortPriority=12}
    False=@n12
   ) {group=Customization, parameterName=Function_Left_Right_Decal_V2}
  )
  n14 = BlendMaterialAttributes(
   A=NamedRerouteUsage()
   B=@n5
   Alpha=StaticSwitchParameter(True=Saturate(None=@n13), False=@n12) {group=Effect-Ice, parameterName=EfIce_CustomMSk_Textures, sortPriority=10}
  ) {vertexAttributeBlendType=UseA}
  n15 = ScalarParameter() {group=Effect-Ice, parameterName="EfIce_Decal_Alpha Masked  1=Active", sortPriority=1}
  n16 = StaticSwitchParameter(True=LinearInterpolate(A=@n6, Alpha=@n15) {constB=0}, False=Constant()) {group=Customization, parameterName=Function_Left_Right_Decal_V2}
  n17 = StaticSwitchParameter(True=OneMinus(None=LinearInterpolate(B=@n6, Alpha=@n15)), False=Constant() {r=1}) {group=Customization, parameterName=Function_Left_Right_Decal_V2}
  n18 = FunctionInput(Preview=TextureSample() {samplerType=SAMPLERTYPE_Masks, texture=/Game/Weapons/FORT_Rifles/Textures/T_Scar_CM}.RGBA) {inputName=CM_Msk, inputType=FunctionInput_Vector4, sortPriority=4}
  n19 = MaterialFunctionCall(Float4=@n18) {materialFunction=&BreakOutFloat4Components}
  n20 = Max(
   A=Min(
    A=ScalarParameter() {defaultValue=1, group=Effect-Ice, parameterName="EfIce_Msk_Primary 1=on", sortPriority=0}
    B=NamedRerouteDeclaration(None=@n19.R) {name=Msk_Prime, nodeColor=<rgba 0 1 0.7422 1>}
   )
   B=Min(
    A=NamedRerouteDeclaration(None=@n19.B) {name=Msk_Accent, nodeColor=<rgba 0 1 0.7656 1>}
    B=ScalarParameter() {group=Effect-Ice, parameterName="EfIce_Msk_Accent 1=on", sortPriority=0}
   )
  )
  n21 = Max(A=@n20, B=ScalarParameter() {group=Effect-Ice, parameterName="EfIce_Msk_Disable 1=on", sortPriority=0})
  n22 = MaterialFunctionCall(Float4=NamedRerouteUsage()) {materialFunction=&BreakOutFloat4Components}
  n23 = NamedRerouteUsage()
  n24 = Saturate(None=LinearInterpolate(
    A=@n21
    B=LinearInterpolate(A=@n20, B=@n21, Alpha=@n22.R)
    Alpha=MaterialFunctionCall(Float4=@n23) {materialFunction=&BreakOutFloat4Components}.A
   ))
  n25 = FunctionInput(Preview=Constant() {r=1}) {bUsePreviewValueAsDefault=true, inputName="Msk_Ts01<--->Ts02", inputType=FunctionInput_Scalar}
  n26 = OneMinus(None=@n25)
  n27 = ScalarParameter() {group=Effect-Ice, parameterName="EfIce_Blend TopSide01---> 02 Invert Msk  1=inverted", sortPriority=1}
  n28 = Constant() {r=1}
  n29 = StaticSwitch(
   True=StaticSwitch(
    True=LinearInterpolate(A=@n25, B=@n26, Alpha=@n27)
    False=@n28
    Value=StaticBoolParameter() {group=Effect-Ice, parameterName="EfIce_Blend TopSide01---> 02 --BlendMask", sortPriority=1}
   )
   False=@n28
   Value=StaticBoolParameter() {group="TopSide_V02_Layer 01 Blend  02", parameterName="TopS Blend 01<--->02"}
  )
  n30 = Min(A=@n24, B=@n29)
  n31 = Saturate(None=Max(A=@n16, B=Min(A=@n17, B=@n30)))
  n32 = FunctionInput(Preview=SetMaterialAttributes!(
    MaterialAttributes
    Normal=TextureSample() {samplerType=SAMPLERTYPE_Normal, texture=/Game/Weapons/FORT_Rifles/T_SCAR_N}.RGB
    BaseColor=Constant3Vector() {constant=<rgba 1 0 0 0>}
   ) {attributeSetTypes [#1, #2]}) {inputName=InMaterial, inputType=FunctionInput_MaterialAttributes}
  n33 = GetMaterialAttributes(None=@n32) {attributeGetTypes [
    #2
    #3
    #1
    #4
   ]}
  FunctionOutput(None=StaticSwitchParameter(True=BlendMaterialAttributes(A=@n5, B=@n14, Alpha=@n31) {vertexAttributeBlendType=UseA}, False=@n33.MaterialAttributes) {group=Customization, parameterName=Effect_Ice}) {outputName=Result_1}
  n34 = GetMaterialAttributes(None=FunctionInput() {inputName=Material_Root, inputType=FunctionInput_MaterialAttributes}) {attributeGetTypes [
    #2
    #3
    #1
    #4
   ]}
  LinearInterpolate(A=@n34.BaseColor, B=@n33.BaseColor, Alpha=ScalarParameter() {group=Effect-Ice, parameterName=EfIce_BaseColorIn Multi Decal, sortPriority=10})
  n35 = TextureCoordinate()
  LinearInterpolate(A=@n35, B=TextureCoordinate() {coordinateIndex=1}, Alpha=ScalarParameter() {group=SimpleIce, parameterName=UseUV2})
  n36 = Add(A=NamedRerouteUsage(), B=Constant2Vector() {r=0.4})
  n37 = ScalarParameter() {defaultValue=4, group=Effect-Ice-Forset, parameterName=EfIce_IceCrackTileRate}
  n38 = NamedRerouteUsage()
  n39 = ChannelMaskParameter(None=TextureSample(UVs=Multiply(A=@n36, B=@n37) {constB=4}, Tex=@n38) {samplerSource=SSM_Wrap_WorldGroupSettings, samplerType=SAMPLERTYPE_Masks}.RGBA) {group=Effect-Ice-Forset, maskChannel=Alpha, parameterName=EfIce_IceCrack_Msk_Channel, defaultValue=<rgba 0 0 0 1>}
  n40 = ScalarParameter() {defaultValue=1.5, group=Effect-Ice-Forset, parameterName=EfIce_IceCrack_Msk_Power}
  n41 = TextureSample(UVs=@n36, Tex=@n38) {samplerSource=SSM_Wrap_WorldGroupSettings}
  n42 = Add(
   A=Multiply(
    A=Power(Base=@n39.R, Exp=@n40) {constExponent=1.5}
    B=ScalarParameter() {defaultValue=0.25, group=Effect-Ice-Forset, parameterName=EfIce_IceCrack_Msk_Strength}
   ) {constB=0.25}
   B=Multiply(
    A=ChannelMaskParameter(None=@n41.RGBA) {group=Effect-Ice-Forset, parameterName=EfIce_Noise_Msk_Channel}.R
    B=ScalarParameter() {defaultValue=0.75, group=Effect-Ice-Forset, parameterName=EfIce_Noise_Msk_Strength}
   ) {constB=0.75}
  )
  n43 = NamedRerouteUsage()
  n44 = Power(
   Base=Fresnel(BaseReflectFractionIn=Constant(), Normal=Transform(None=@n43)) {exponent=1}
   Exp=ScalarParameter() {defaultValue=1, group=Effect-Ice_Highlight, parameterName=EfIce_HighlightFresnelExponent}
  ) {constExponent=1.2}
  n45 = ScalarParameter() {defaultValue=4, group=Effect-Ice_Highlight, parameterName=EfIce_HighLightEmissive}
  NamedRerouteDeclaration(None=Multiply(A=@n42, B=Multiply(A=@n44, B=@n45))) {name=Edge highlight, nodeColor=<rgba 1 0 0.4922 1>}
  n46 = FunctionInput() {inputName=Multi Decal, inputType=FunctionInput_MaterialAttributes}
  n47 = ScalarParameter() {
   desc="When using disable Blends 01--02 on Decals to work correctly. "
   group=Effect-Ice
   parameterName=EfIce_BaseMat Strait Topside
   sortPriority=10
  }
  n48 = Min(A=@n47, B=NamedRerouteDeclaration(None=@n29) {name="TS01--->02_Msk", nodeColor=<rgba 0.2188 0 1 1>})
  n49 = GetMaterialAttributes(None=StaticSwitch(
    True=BlendMaterialAttributes(A=@n32, B=@n46, Alpha=@n48) {vertexAttributeBlendType=UseA}
    False=@n32
    Value=StaticBoolParameter() {group=TopSide_V02, parameterName=TopSideProjection_V02}
   )) {attributeGetTypes [
    #2
    #3
    #1
    #4
    #5
    #6
   ]}
  NamedRerouteDeclaration(None=@n49.Specular) {name=InSpecular, nodeColor=<rgba 0.0312 0 1 1>}
  n50 = VertexNormalWS()
  n51 = CameraVectorWS()
  n52 = DotProduct(A=Normalize(VectorInput=Add(A=@n50, B=@n51)), B=Multiply(A=AtmosphericLightVector()) {constB=-1})
  n53 = VertexNormalWS()
  n54 = CameraVectorWS()
  n55 = Multiply(A=Add(A=Multiply(A=@n52) {constB=0.5}) {constB=0.5}, B=Saturate(None=DotProduct(A=@n53, B=@n54)))
  n56 = ScalarParameter() {defaultValue=2, group=Effect-Ice-Surface, parameterName=EfIce_FakeSSSExponent}
  NamedRerouteDeclaration(None=Saturate(None=Power(Base=@n55, Exp=@n56))) {name=FakeSSS, nodeColor=<rgba 0 0.0859 1 1>}
  n57 = MaterialFunctionCall!(
   Ray Direction=Multiply(A=CameraVectorWS()) {constB=-1}
   Surface Normal=PixelNormalWS()
   Refractive Index Origin
   Refractive Index Target=ScalarParameter() {defaultValue=3, group=Effect-Ice_Reflection, parameterName=EfIce_RefractionAmount}
  ) {materialFunction=/Engine/Functions/Engine_MaterialFunctions01/Vectors/Refract}
  n58 = Multiply(
   A=Subtract(A=NamedRerouteUsage(), B=ScalarParameter() {defaultValue=0.5, group=Effect-Ice_Reflection, parameterName=EfIce_RefractionForset_Offset}) {constB=0.5}
   B=ScalarParameter() {group=Effect-Ice_Reflection, parameterName=EfIce_RefractionForset_Strength}
  ) {constB=0}
  NamedRerouteDeclaration(None=SkyLightEnvMapSample(Direction=Add(A=@n57.Vector, B=@n58))) {name=EnvironmentMap, nodeColor=<rgba 0 1 0.0625 1>}
  n59 = Constant4Vector()
  NamedRerouteDeclaration(None=FunctionInput(Preview=@n59.RGBA) {bUsePreviewValueAsDefault=true, inputName=IsState, inputType=FunctionInput_Vector4, sortPriority=4}) {name=IsState, nodeColor=<rgba 0.1497 1 0.1633 1>}
  NamedRerouteDeclaration(None=FunctionInput(Preview=@n59.RGBA) {bUsePreviewValueAsDefault=true, inputName=IsState_B, inputType=FunctionInput_Vector4, sortPriority=4}) {name=IsState_B, nodeColor=<rgba 0.1497 1 0.1633 1>}
  n60 = PreSkinnedLocalBounds()
  n61 = Subtract(A=VertexInterpolator(VS=PreSkinnedPosition()).PS, B=@n60.Min)
  n62 = MaterialFunctionCall(Color=Subtract(A=@n60.Max, B=@n60.Max)) {materialFunction=&SplitComponents}
  n63 = Max(A=Max(A=@n62.R, B=@n62.R), B=@n62.B)
  NamedRerouteDeclaration(None=FunctionInput(Preview=Divide(A=@n61, B=@n63)) {bUsePreviewValueAsDefault=true, inputName=Calcated_Bounds, sortPriority=11}) {name=Caclcated_Bounds, nodeColor=<rgba 0 0.9297 1 1>}
  n64 = PreSkinnedNormal()
  NamedRerouteDeclaration(None=FunctionInput(Preview=VertexInterpolator(VS=@n64).PS) {bUsePreviewValueAsDefault=true, inputName=Pre-Skinned Local Normal, sortPriority=10}) {name=PreSkin_Normal, nodeColor=<rgba 0.8281 1 0 1>}
  NamedRerouteDeclaration(None=@n49.EmissiveColor) {name=InEmissive, nodeColor=<rgba 1 0 0.3984 1>}
  n65 = Saturate(None=Multiply(A=@n42) {constB=1.5})
  NamedRerouteDeclaration(None=@n65) {name=Forset, nodeColor=<rgba 1 0 0.1641 1>}
  NamedRerouteDeclaration(None=@n49.MaterialAttributes) {name=In_Material, nodeColor=<rgba 0.7812 1 0 1>}
  NamedRerouteDeclaration(None=@n19.G) {name=Msk_Curvature, nodeColor=<rgba 0 1 0.7422 1>}
  NamedRerouteDeclaration(None=@n18) {name=CM_Mask, nodeColor=<rgba 1 0 0.3047 1>}
  n66 = VectorParameter() {
   group=Effect-Ice-Surface
   parameterName="EfIce_Surfac_DeepColor  Gmap\\Curvature"
   sortPriority=41
   defaultValue=<rgba 0.0237 0.0762 0.1198 1>
  }
  n67 = MaterialFunctionCall(Float3=Abs(None=NamedRerouteUsage())) {materialFunction=&BreakOutFloat3Components}
  n68 = Constant() {r=0.5}
  n69 = NamedRerouteUsage()
  n70 = ComponentMask(None=@n69) {b=true, desc=YZ, g=true}
  n71 = ComponentMask(None=@n69) {b=true, desc=XZ, r=true}
  n72 = NamedRerouteDeclaration(None=If(A=@n67.B, B=@n68, "A > B"=ComponentMask(None=@n69) {desc=YX, g=true, r=true}, "A < B"=If(A=@n67.R, B=@n68, "A > B"=@n70, "A < B"=@n71))) {name=UVS_Cheap_Tri, nodeColor=<rgba 1 0 0.8438 1>}
  n73 = ScalarParameter() {defaultValue=2, group=Effect-Ice-Surface, parameterName=EfIce_SurfacNoiseTileRate}
  n74 = ChannelMaskParameter(None=TextureSample(UVs=Multiply(A=@n72, B=@n73), Tex=NamedRerouteUsage()) {samplerSource=SSM_Wrap_WorldGroupSettings, samplerType=SAMPLERTYPE_Grayscale}.RGBA) {group=Effect-Ice-Surface, maskChannel=Green, parameterName=EfIce_SurfacNoise_Msk_Channel, defaultValue=<rgba 0 1 0 0>}
  n75 = ScalarParameter() {defaultValue=1.2, group=Effect-Ice-Surface, parameterName=EfIce_SurfacNoiseExp}
  n76 = Multiply(
   A=LinearInterpolate(Alpha=Power(Base=@n74.R, Exp=@n75) {constExponent=1.2}) {constA=0.3}
   B=VectorParameter() {group=Effect-Ice-Surface, parameterName=EfIce_Surfac_Color, defaultValue=<rgba 0.0652 0.1684 0.3229 1>}.RGB
  )
  n77 = NamedRerouteUsage()
  n78 = ComponentMask(None=@n77) {g=true}
  n79 = VectorParameter() {
   group=Effect-Ice-Surface
   parameterName=EfIce_Surface_Curavture
   sortPriority=42
   channelNames {a=Intenstiy, b=Exponent, g=Pits, r=Peaks}
   defaultValue=<rgba 1 0 1 1>
  }
  n80 = Multiply(A=Saturate(None=Add(A=@n78) {constB=-0.5}), B=@n79.Peaks)
  n81 = Add(A=@n78) {constB=0.5}
  n82 = Multiply(A=OneMinus(None=Saturate(None=@n81)), B=@n79.Pits)
  n83 = Multiply(A=Power(Base=Add(A=@n80, B=@n82), Exp=@n79.Exponent) {constExponent=5}, B=@n79.Intenstiy)
  n84 = ScalarParameter() {defaultValue=1, group=Effect-Ice-Surface, parameterName=EfIce_Surface_GmapIntenstiy, sortPriority=41}
  n85 = Power(
   Base=ComponentMask(None=@n77) {a=true}
   Exp=ScalarParameter() {defaultValue=2, group=Effect-Ice-Surface, parameterName=EfIce_Surface_GmapExponent, sortPriority=41}
  ) {constExponent=5}
  n86 = Saturate(None=StaticSwitchParameter(True=OneMinus(None=@n83), False=Multiply(A=@n84, B=@n85)) {group=Effect-Ice-Surface, parameterName="EfIce_Surface_Gmap-------->Curvature", sortPriority=40})
  NamedRerouteDeclaration(None=Multiply(
    A=LinearInterpolate(A=@n66.RGB, B=@n76, Alpha=@n86)
    B=ScalarParameter() {defaultValue=1, group=Effect-Ice-Surface, parameterName=EfIce_Surfac_ColorIntensity}
   )) {name=ICE Color, nodeColor=<rgba 0.6875 1 0 1>}
  n87 = LinearInterpolate(
   A=NamedRerouteUsage()
   B=LinearInterpolate(
    A=NamedRerouteUsage()
    B=NamedRerouteUsage()
    Alpha=ScalarParameter() {group=Effect-Ice, parameterName=EfIce_BaseColorRoot-Amount, sortPriority=10}
   )
   Alpha=Saturate(None=ScalarParameter() {group=Effect-Ice, parameterName=EfIce_BaseColorIn-Blend, sortPriority=10})
  ) {constAlpha=0.9}
  n88 = ScalarParameter() {defaultValue=1, group=Effect-Ice, parameterName=EfIce_BaseColorIntensity}
  n89 = NamedRerouteUsage()
  n90 = ScalarParameter() {defaultValue=1, group=Effect-Ice, parameterName=EfIce_SpecularMultiplier, sortPriority=50}
  n91 = Multiply(
   A=Add(A=NamedRerouteUsage(), B=NamedRerouteUsage())
   B=ScalarParameter() {defaultValue=1, group=Effect-Ice, parameterName=EfIce_Emissive Strength, sortPriority=50}
  )
  NamedRerouteDeclaration(None=MakeMaterialAttributes!(
    BaseColor=Multiply(A=@n87, B=@n88)
    Metallic=ScalarParameter() {group=Effect-Ice, parameterName=EfIce_Metalness, sortPriority=50}
    Specular=Multiply(A=@n89, B=@n90)
    Roughness=ScalarParameter() {defaultValue=0.1, group=Effect-Ice, parameterName=EfIce_Roughness, sortPriority=50}
    Anisotropy
    EmissiveColor=Clamp(None=@n91) {maxDefault=100}
    Opacity=Constant() {r=1}
    OpacityMask
    Normal=NamedRerouteUsage()
    Tangent
    WorldPositionOffset
    SubsurfaceColor
    ClearCoat
    ClearCoatRoughness
    AmbientOcclusion=Constant() {r=1}
    Refraction
    CustomizedUVs_0
    CustomizedUVs_1
    CustomizedUVs_2
    CustomizedUVs_3
    CustomizedUVs_4
    CustomizedUVs_5
    CustomizedUVs_6
    CustomizedUVs_7
    PixelDepthOffset
    ShadingModel
    Displacement
   )) {name=Mat_Base, nodeColor=<rgba 0 0.3906 1 1>}
  n92 = Multiply(A=@n35, B=ScalarParameter() {defaultValue=0.5, group=Effect-Ice-inner, parameterName=EfIce_InnerIce_TileReate}) {constB=2.5}
  n93 = ChannelMaskParameter(None=TextureSample(UVs=Multiply(A=@n92), Tex=NamedRerouteUsage()) {samplerSource=SSM_Wrap_WorldGroupSettings}.RGBA) {group=Effect-Ice-inner, parameterName=EfIce_InnerIce_Mask_Channel}
  n94 = MaterialFunctionCall!(
   Normal=MaterialFunctionCall(Normal=NamedRerouteUsage(), Flatness=Constant() {r=0.5}) {materialFunction=&FlattenNormal}.Result
   CameraVector=Transform(None=CameraVectorWS()) {transformSourceType=TRANSFORMSOURCE_World, transformType=TRANSFORM_Tangent}
  ) {materialFunction=/Engine/Functions/Engine_MaterialFunctions02/Utility/CustomReflectionVector}
  n95 = ComponentMask(None=@n94.Reflection Vector) {b=true}
  n96 = Multiply(
   A=ComponentMask(None=@n94.Reflection Vector) {g=true, r=true}
   B=Divide(A=ScalarParameter() {defaultValue=20, group=Effect-Ice-inner, parameterName=EfIce_InnerIceDepth}, B=Abs(None=@n95)) {constA=100}
  )
  n97 = Divide(B=Constant() {r=1600}) {constB=512}
  NamedRerouteDeclaration(None=Add(A=BumpOffset(Coordinate=@n92, Height=@n93.R) {heightRatio=0.1}, B=Multiply(A=@n96, B=@n97))) {name=Ice uv, nodeColor=<rgba 0.7578 1 0 1>}
  n98 = TextureObjectParameter() {
   group=TopSide_V02
   parameterName=TopS_00A_Sharerd_Tex_Msk
   samplerType=SAMPLERTYPE_Masks
   sortPriority=249
   texture=&T_BlackMask
  }
  n99 = TextureObjectParameter() {
   group=TopSide_V02
   parameterName=TopS_00B_Sharerd_Tex_Msk
   samplerType=SAMPLERTYPE_Masks
   sortPriority=249
   texture=&T_BlackMask
  }
  n100 = NamedRerouteDeclaration(None=@n99) {name=Tex_00B_Shared_Tex, nodeColor=<rgba 1 0.8906 0 1>}
  n101 = TextureObjectParameter() {group=Effect-Ice, parameterName=EfIce_RMsk--Tex, samplerType=SAMPLERTYPE_Masks, texture=&T_BlackMask}
  NamedRerouteDeclaration(None=StaticSwitchParameter(
    True=NamedRerouteDeclaration(None=@n98) {name=Tex_00A_Shared_Tex, nodeColor=<rgba 1 0.8906 0 1>}
    False=StaticSwitchParameter(True=@n100, False=@n101) {group=Effect-Ice, parameterName="EfIce_RMsk------>TopS_00B_Sharerd"}
   ) {group=Effect-Ice, parameterName="EfIce_RMsk------>TopS_00A_Sharerd"}) {name=RMask, nodeColor=<rgba 0 1 0.5313 1>}
  NamedRerouteDeclaration(None=@n34.MaterialAttributes) {name=Root_Material, nodeColor=<rgba 0.7812 1 0 1>}
  n102 = NamedRerouteUsage()
  n103 = ScalarParameter() {defaultValue=1, group=Effect-Ice_Reflection, parameterName=EfIce_BaseColorEnvironmentAmount}
  n104 = NamedRerouteUsage()
  n105 = ScalarParameter() {defaultValue=2, group=Effect-Ice_Reflection, parameterName=EfIce_BaseColorForsetExp}
  n106 = Saturate(None=Multiply(
    A=Power(Base=@n104, Exp=@n105) {constExponent=1}
    B=ScalarParameter() {defaultValue=2, group=Effect-Ice_Reflection, parameterName=EfIce_BaseColorForsetStrength}
   ))
  n107 = Add(A=LinearInterpolate(A=Multiply(A=@n102, B=@n103) {constB=0.5}, Alpha=@n106) {constAlpha=0.1, constB=0}, B=@n106)
  n108 = NamedRerouteUsage()
  n109 = ScalarParameter() {defaultValue=0.15, group=Effect-Ice-Frost, parameterName=EfIce_Frost_Msk_Strength}
  n110 = TextureSample(
   UVs=Multiply(A=TextureCoordinate(), B=ScalarParameter() {defaultValue=2, group=Effect-Ice-Frost, parameterName=EfIce_FrostTileRate})
   Tex=NamedRerouteUsage()
  ) {samplerSource=SSM_Wrap_WorldGroupSettings}
  n111 = NamedRerouteDeclaration(None=Power(
    Base=ChannelMaskParameter(None=@n110.RGBA) {group=Effect-Ice-Frost, parameterName=EfIce_Frost_Msk_Channel}.R
    Exp=ScalarParameter() {defaultValue=3, group=Effect-Ice-Frost, parameterName=EfIce_Frost_Msk_Power}
   ) {constExponent=3}) {name=Frost mask, nodeColor=<rgba 0.4531 1 0 1>}
  NamedRerouteDeclaration(None=LinearInterpolate(
    A=Multiply(A=@n107, B=@n108)
    B=VectorParameter() {group=Effect-Ice-Frost, parameterName=EfIce_Frost_Msk_Color, defaultValue=<rgba 1 1 1 1>}.RGB
    Alpha=Multiply(A=@n109, B=@n111) {constA=0.15}
   )) {name=Base Color, nodeColor=<rgba 0.0781 1 0 1>}
  n112 = NamedRerouteUsage()
  NamedRerouteDeclaration(None=LinearInterpolate(A=ScalarParameter() {group=Effect-Ice_Highlight, parameterName=EfIce_IceDepthColorIntensity}, Alpha=Max(A=@n65, B=@n112))) {name=IceHighlight, nodeColor=<rgba 0.2656 1 0 1>}
  NamedRerouteDeclaration(None=@n34.BaseColor) {name=RootBaseColor, nodeColor=<rgba 0.875 1 0 1>}
  n113 = SkyAtmosphereLightDirection()
  n114 = CameraVectorWS()
  n115 = MaterialFunctionCall!(
   Normal=PixelNormalWS()
   Light Vector=Normalize(VectorInput=Add(A=@n113, B=@n114))
   Roughness=ScalarParameter() {defaultValue=0.05, group=Effect-Ice-Surface, parameterName=EfIce_Fake_Spec_Roughness, sortPriority=10}
  ) {materialFunction=/Engine/Functions/Engine_MaterialFunctions01/Lighting/GGXSpecular}
  n116 = Saturate(None=Multiply(A=@n115.Result, B=@n115.Result))
  n117 = ScalarParameter() {defaultValue=1, group=Effect-Ice-Surface, parameterName=EfIce_Fake_Spec_Brightness, sortPriority=10}
  n118 = LinearInterpolate(
   B=ScalarParameter() {defaultValue=1.8, group=Effect-Ice_Highlight, parameterName=EfIce_RoughnessHighlightIntensity}
   Alpha=NamedRerouteUsage()
  ) {constB=1.8}
  n119 = Multiply(A=LinearInterpolate(Alpha=NamedRerouteUsage()) {constA=0.8}, B=LinearInterpolate(Alpha=@n118) {constA=0.05})
  n120 = VectorParameter() {group=Effect-Ice-Surface, parameterName=EfIce_Emissive color, defaultValue=<rgba 0.2611 0.5021 1 1>}
  n121 = Multiply(A=@n120.RGB) {constB=1.5}
  n122 = Power(Base=@n118, Exp=ScalarParameter() {defaultValue=8, group=Effect-Ice_Highlight, parameterName=EfIce_RoughnessHighlightExp}) {constExponent=8}
  n123 = LinearInterpolate(A=Multiply(A=@n3, B=@n120.RGB), B=Add(A=Multiply(A=@n119, B=@n120.RGB), B=Multiply(A=@n121, B=@n122)), Alpha=Constant() {r=0.5}) {constAlpha=0.7}
  n124 = LinearInterpolate(
   A=@n123
   B=Multiply(A=@n123, B=VectorParameter() {group=Effect-Ice-Surface, parameterName=EfIce_FakeSSSColor, defaultValue=<rgba 0.4208 2.415 5 1>}.RGB) {constB=3}
   Alpha=NamedRerouteUsage()
  )
  n125 = Constant() {r=0.8}
  n126 = MaterialFunctionCall(Color_Morning=@n125, Color_Day=Constant() {r=1}, Color_Evening=@n125, Color_Night=Constant() {r=0.6}) {materialFunction=&Color_Time_of_Day}
  NamedRerouteDeclaration(None=Add(A=Multiply(A=@n116, B=@n117), B=Multiply(A=@n124, B=@n126.Result))) {name=Emissive Color, nodeColor=<rgba 1 0.3516 0 1>}
  NamedRerouteDeclaration(None=@n33.BaseColor) {name=InBaseColor, nodeColor=<rgba 0.875 1 0 1>}
  NamedRerouteDeclaration(None=@n49.Normal) {name=InNormalMap, nodeColor=<rgba 0.8047 0 1 1>}
  n127 = Max(A=Min(A=Min(A=@n24, B=@n29), B=@n17), B=@n16)
  NamedRerouteDeclaration(None=LinearInterpolate(A=Constant() {r=0.1}, B=@n49.Roughness, Alpha=Saturate(None=@n127))) {name=InRoughness, nodeColor=<rgba 1 0.3516 0 1>}
 ]}
