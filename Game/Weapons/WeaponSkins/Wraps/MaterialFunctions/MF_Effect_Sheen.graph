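# Material function for weapon-wrap skins that adds an animated sheen highlight to base color and emissive when the Effect_Sheen static switch is on.
# A scrolling texture is sampled on side-normal projected UVs shifted by view direction, normal warp and per-item-type scale and rotation.
# The highlight is weighted by roughness, metallic or specular and the gradient mask, then tinted.
# The result blends over the input material through Primary/Accent and optional multi-decal masks.
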MaterialFunction /Game/Weapons/WeaponSkins/Wraps/MaterialFunctions/MF_Effect_Sheen
graph {nodes [
  n1 = NamedRerouteUsage()
  n2 = NamedRerouteUsage()
  n3 = VectorParameter() {
   group=Effect Sheen
   parameterName=ESheen_Decal_Multi
   sortPriority=1
   channelNames {a="DecalD     1-On", b="DecalC     1-On", g="DecalB     1-On", r="DecalA     1-On"}
   defaultValue=<rgba 1 1 1 1>
  }
  MaterialExpressionFunctionInput_7 = FunctionInput(Preview=Constant4Vector().RGBA) {bUsePreviewValueAsDefault=true, inputName=multi_Decal, inputType=FunctionInput_Vector4}
  n4 = Convert(None=@MaterialExpressionFunctionInput_7) {
   nodeName=Break Float4
   convertInputs [{
     type=Vector4
     defaultValue=<rgba 0 0 0 1>
     expressionInput {
      expression=&self:MaterialExpressionFunctionInput_7
      inputName=None
      mask=0
      maskA=0
      maskB=0
      maskG=0
      maskR=0
      outputIndex=0
     }
    }]
   convertOutputs [{type=Scalar}, {type=Scalar}, {type=Scalar}, {type=Scalar}]
  }
  n5 = Max(A=Min(A=@n3."DecalA     1-On", B=@n4), B=Min(A=@n3."DecalB     1-On", B=@n4))
  n6 = Max(A=Min(A=@n3."DecalC     1-On", B=@n4), B=Min(A=@n3."DecalD     1-On", B=@n4))
  MaterialExpressionOneMinus_3 = OneMinus(None=@MaterialExpressionFunctionInput_7)
  n7 = Convert(None=@MaterialExpressionOneMinus_3) {
   nodeName=Break Float4
   convertInputs [{
     type=Vector4
     defaultValue=<rgba 0 0 0 1>
     expressionInput {
      expression=&self:MaterialExpressionOneMinus_3
      inputName=None
      mask=0
      maskA=0
      maskB=0
      maskG=0
      maskR=0
      outputIndex=0
     }
    }]
   convertOutputs [{type=Scalar}, {type=Scalar}, {type=Scalar}, {type=Scalar}]
  }
  n8 = Min(
   A=Min(A=Max(A=@n3."DecalA     1-On", B=@n7), B=Max(A=@n3."DecalB     1-On", B=@n7))
   B=Min(A=Max(A=@n3."DecalC     1-On", B=@n7), B=Max(A=@n3."DecalD     1-On", B=@n7))
  )
  n9 = Constant() {r=1}
  n10 = MaterialFunctionCall(Float4=FunctionInput(Preview=TextureSample() {samplerType=SAMPLERTYPE_Masks, texture=/Game/Weapons/FORT_Rifles/Textures/T_Scar_CM}.RGBA) {inputName=CM_Msk, inputType=FunctionInput_Vector4, sortPriority=4}) {materialFunction=&BreakOutFloat4Components}
  n11 = VectorParameter() {
   group=Effect Sheen
   parameterName="ESheen_Primary\\Accent"
   sortPriority=0
   channelNames {a="", b="disable ", g=Accent, r=Primary}
   defaultValue=<rgba 1 0 0 1>
  }
  n12 = Max(
   A=Min(A=NamedRerouteDeclaration(None=@n10.R) {name=Msk_Prime, nodeColor=<rgba 0 1 0.7422 1>}, B=@n11.Primary)
   B=Min(A=NamedRerouteDeclaration(None=@n10.B) {name=Msk_Accent_1, nodeColor=<rgba 0 1 0.7656 1>}, B=@n11.Accent)
  )
  n13 = Max(
   A=StaticSwitchParameter(True=Max(A=@n5, B=@n6), False=Constant()) {group=Customization, parameterName=Effect_Multi_Decal_Reactive}
   B=Min(
    A=StaticSwitchParameter(True=@n8, False=@n9) {group=Customization, parameterName=Effect_Multi_Decal_Reactive}
    B=Max(A=@n12, B=@n11."disable ")
   )
  )
  n14 = TextureSample() {samplerType=SAMPLERTYPE_Normal, texture=/Game/Weapons/FORT_Rifles/T_SCAR_N}
  n15 = Constant3Vector() {constant=<rgba 1 0 0 0>}
  n16 = GetMaterialAttributes(None=FunctionInput(Preview=SetMaterialAttributes!(MaterialAttributes, Normal=@n14.RGB, BaseColor=@n15) {attributeSetTypes [#1, #2]}) {inputName=Material, inputType=FunctionInput_MaterialAttributes}) {attributeGetTypes [
    #2
    #3
    #1
    #4
    #5
    #6
   ]}
  FunctionOutput(None=StaticSwitchParameter(
    True=BlendMaterialAttributes(A=@n1, B=@n2, Alpha=@n13) {vertexAttributeBlendType=UseA}
    False=NamedRerouteDeclaration(None=@n16.MaterialAttributes) {name=Root_Material, nodeColor=<rgba 0.7812 1 0 1>}
   ) {group=Customization, parameterName=Effect_Sheen}) {outputName=Result_1}
  MaterialFunctionCall(Float2=NamedRerouteUsage()) {materialFunction=&BreakOutFloat2Components}
  MaterialExpressionDivide_1 = Divide(
   A=TransformPosition(None=WorldPosition().XYZ) {transformSourceType=TRANSFORMPOSSOURCE_World}
   B=ScalarParameter() {defaultValue=30, group=Effect Sheen, parameterName=ESheen_SpecularHighlight_Planar_Scale, sliderMax=200}
  ) {constB=30}
  n17 = Convert(None=@MaterialExpressionDivide_1) {
   nodeName=Break Float3
   convertInputs [{
     type=Vector3
     expressionInput {
      expression=&self:MaterialExpressionDivide_1
      inputName=None
      mask=0
      maskA=0
      maskB=0
      maskG=0
      maskR=0
      outputIndex=0
     }
    }]
   convertOutputs [{type=Vector2}, {type=Vector2}, {type=Vector2}]
  }
  n18 = StaticSwitchParameter(True=@n17, False=@n17) {defaultValue=true, desc="True = Y plane\r\nFalse = X plane", group=Effect Sheen, parameterName=ESheen_UsePlanarReflectionYPlane}
  Multiply(A=StaticSwitchParameter(True=@n17, False=@n18) {group=Effect Sheen, parameterName=ESheen_UsePlanarReflectionZPlane}) {constB=-1}
  n19 = Constant4Vector()
  NamedRerouteDeclaration(None=FunctionInput(Preview=@n19.RGBA) {bUsePreviewValueAsDefault=true, inputName=IsState_B, inputType=FunctionInput_Vector4, sortPriority=4}) {name=IsState_B, nodeColor=<rgba 0.1497 1 0.1633 1>}
  n20 = PreSkinnedLocalBounds()
  n21 = Subtract(A=VertexInterpolator(VS=PreSkinnedPosition()).PS, B=@n20.Min)
  n22 = MaterialFunctionCall(Color=Subtract(A=@n20.Max, B=@n20.Max)) {materialFunction=&SplitComponents}
  n23 = Max(A=Max(A=@n22.R, B=@n22.R), B=@n22.B)
  NamedRerouteDeclaration(None=FunctionInput(Preview=Divide(A=@n21, B=@n23)) {bUsePreviewValueAsDefault=true, inputName=Calcated_Bounds, sortPriority=11}) {name=Caclcated_Bounds, nodeColor=<rgba 0 0.9297 1 1>}
  n24 = PreSkinnedNormal()
  NamedRerouteDeclaration(None=FunctionInput(Preview=VertexInterpolator(VS=@n24).PS) {bUsePreviewValueAsDefault=true, inputName=Pre-Skinned Local Normal, sortPriority=10}) {name=PreSkin_Normal, nodeColor=<rgba 0.8281 1 0 1>}
  n25 = ScalarParameter() {defaultValue=1, group=Effect Sheen_UVs, parameterName=ESheen_Scale_Vehicle, sortPriority=150}
  n26 = ScalarParameter() {defaultValue=1, group=Effect Sheen_UVs, parameterName=ESheen_Scale_Vehicle_Side, sortPriority=150}
  n27 = MaterialFunctionCall(Float4=NamedRerouteUsage()) {materialFunction=&BreakOutFloat4Components}
  n28 = LinearInterpolate(
   A=LinearInterpolate(
    A=ScalarParameter() {defaultValue=1, group=Effect Sheen_UVs, parameterName=ESheen_Scale_Weapon, sortPriority=150}
    B=LinearInterpolate(A=@n25, B=@n26, Alpha=@n27.G)
    Alpha=@n27.R
   )
   B=LinearInterpolate(
    A=ScalarParameter() {defaultValue=1, group=Effect Sheen_UVs, parameterName=ESheen_Scale_Mech, sortPriority=150}
    B=ScalarParameter() {defaultValue=1, group=Effect Sheen_UVs, parameterName=ESheen_Scale_Character, sortPriority=150}
    Alpha=@n27.A
   )
   Alpha=@n27.B
  )
  n29 = ScalarParameter() {defaultValue=1, group=Effect Sheen_UVs, parameterName=ESheen_Scale_Backpack, sortPriority=150}
  n30 = MaterialFunctionCall(Float2=NamedRerouteUsage()) {materialFunction=&BreakOutFloat2Components}
  NamedRerouteDeclaration(None=LinearInterpolate(
    A=LinearInterpolate(A=@n28, B=@n29, Alpha=@n30.R)
    B=ScalarParameter() {defaultValue=1, group=Effect Sheen_UVs, parameterName=ESheen_Scale_Pickaxe, sortPriority=150}
    Alpha=@n30.G
   )) {name=UVs_Scale_V1, nodeColor=<rgba 0 0.7188 1 1>}
  n31 = ScalarParameter() {defaultValue=0.1, group=Effect Sheen_UVs, parameterName="ESheen_Rotation_Weapon\\Pickaxe", sortPriority=150}
  n32 = MaterialFunctionCall(Float4=NamedRerouteUsage()) {materialFunction=&BreakOutFloat4Components}
  n33 = LinearInterpolate(
   A=ScalarParameter() {defaultValue=0.1, group=Effect Sheen_UVs, parameterName=ESheen_Rotation_Vehicle, sortPriority=150}
   B=ScalarParameter() {defaultValue=0.1, group=Effect Sheen_UVs, parameterName=ESheen_Rotation_Vehicle_Side, sortPriority=150}
   Alpha=@n32.G
  )
  NamedRerouteDeclaration(None=LinearInterpolate(
    A=LinearInterpolate(A=@n31, B=@n33, Alpha=@n32.R)
    B=ScalarParameter() {defaultValue=0.1, group=Effect Sheen_UVs, parameterName="ESheen_Rotation_Mech\\Character\\Back", sortPriority=150}
    Alpha=@n32.B
   )) {name=UVs_Rotation_1, nodeColor=<rgba 0 0.7188 1 1>}
  n34 = TextureObjectParameter() {
   group=TopSide_V02
   parameterName=TopS_00B_Sharerd_Tex_Msk
   samplerType=SAMPLERTYPE_Masks
   sortPriority=249
   texture=&T_BlackMask
  }
  NamedRerouteDeclaration(None=@n34) {name=Tex_00B_Shared_Tex, nodeColor=<rgba 1 0.8906 0 1>}
  NamedRerouteDeclaration(None=@n10.A) {name=Msk_Gmap, nodeColor=<rgba 0 1 0.7656 1>}
  n35 = NamedRerouteUsage()
  n36 = VectorParameter() {
   group=Effect Sheen
   parameterName=ESheen_Specular_Channel Adjustmens
   channelNames {a="", b=High Values, g="Low values ", r=""}
   defaultValue=<rgba 0 0 1 1>
  }
  n37 = VectorParameter() {
   group=Effect Sheen
   parameterName="ESheen_Specular_Channel  Influnce"
   channelNames {a="", b="Specular     =1", g="Metallic      =1", r=" Roughness =1"}
   defaultValue=<rgba 0 0 1 1>
  }
  n38 = LinearInterpolate(B=@n16.Roughness, Alpha=@n37." Roughness =1")
  n39 = LinearInterpolate(
   A=@n36."Low values "
   B=@n36."Low values "
   Alpha=LinearInterpolate(A=LinearInterpolate(A=@n38, B=@n16.Metallic, Alpha=@n37."Metallic      =1"), B=@n16.Specular, Alpha=@n37."Specular     =1")
  )
  n40 = Transform(None=Constant3Vector() {constant=<rgba 0 0 1 0>}) {transformSourceType=TRANSFORMSOURCE_View, transformType=TRANSFORM_Local}
  n41 = ComponentMask(None=@n40) {g=true, r=true}
  n42 = MaterialFunctionCall(Float2=MaterialFunctionCall(Vector or UVs=Abs(None=@n41)) {materialFunction=&VectorToRadialValue}.Radial Coordinates) {materialFunction=&BreakOutFloat2Components}
  n43 = ScalarParameter() {defaultValue=4, group=Effect Sheen, parameterName=ESheen_Offset_Strength}
  n44 = AppendVector(A=Add(A=Multiply(A=@n42.R, B=@n43) {constB=4}) {constB=-1}, B=Multiply(A=ComponentMask(None=@n40) {b=true}) {constB=-1})
  n45 = Add(A=MaterialFunctionCall!(
    Normal_Map
    Offset Rt
    Pivot
    Caculated_Bounds=NamedRerouteUsage()
    Pre_Skinned_Normals=NamedRerouteUsage()
    is_State=NamedRerouteUsage()
    UV_Rotation
    Uvs_Offset
    Uvs_TileRate=NamedRerouteUsage()
    Uvs_Offset_Pivot
    For_Safe_Zones
    Rotation_Comp
    Clamp Y
    Clamp X
    Clamp Padding_Comp
    No_Rotation=StaticBool() {value=true}
   ) {materialFunction=&/MF_Projection_Side_Normal_V2}.UVs) {constB=-0.5}
  n46 = Add(A=NamedRerouteUsage(), B=Rotator(Coordinate=Add(A=@n44, B=@n45), Time=NamedRerouteUsage()) {speed=3.1415})
  n47 = ComponentMask(None=Multiply(A=VectorParameter() {group=Effect Sheen_UVs, parameterName=ESheen_Specular Speed, defaultValue=<rgba 0 0 0 1>}.RGB, B=Time())) {g=true, r=true}
  n48 = NamedRerouteUsage()
  n49 = StaticSwitchParameter(
   True=NamedRerouteUsage()
   False=TextureObjectParameter() {group=Effect Sheen, parameterName=ESheen_Specular Tex, texture=&T_BlacksRGB}
  ) {group=Effect Sheen, parameterName="ESheen_Specular--------->TopS_00B_Msk"}
  n50 = ChannelMaskParameter(None=TextureSample(
    UVs=Add(A=@n46, B=@n47)
    Tex=StaticSwitchParameter(True=@n48, False=@n49) {group=Effect Sheen, parameterName="ESheen_Specular--------->TopS_00A_Msk"}
   ).RGBA) {group=Effect Sheen, parameterName=ESheen_Specular Tex_Channel}
  n51 = VectorParameter() {group=Effect Sheen, parameterName=ESheen_Gmap, channelNames {a="", b=High, g=Low, r=Use}, defaultValue=<rgba 0 0.2 1 1>}
  n52 = LinearInterpolate(
   B=MaterialFunctionCall(Input=NamedRerouteUsage(), Input Low=@n51.Low, Input High=@n51.Low) {materialFunction=&RemapValue_0-1}.Result
   Alpha=@n51.Use
  ) {constA=1}
  n53 = Multiply(A=Multiply(A=@n39, B=@n50.R), B=@n52)
  n54 = OneMinus(None=Min(A=@n53, B=ScalarParameter() {group=Effect Sheen, parameterName=ESheen_Darkeen BaseColor}))
  n55 = Multiply(
   A=@n50.R
   B=MaterialFunctionCall(
    A=VectorParameter() {group=Effect Sheen, parameterName=ESheen_Speculal_Color-A, defaultValue=<rgba 1 1 1 1>}.RGB
    B=VectorParameter() {group=Effect Sheen, parameterName=ESheen_Speculal_Color-B, defaultValue=<rgba 1 1 1 1>}.RGB
    C=VectorParameter() {group=Effect Sheen, parameterName=ESheen_Speculal_Color-C, defaultValue=<rgba 1 1 1 1>}.RGB
    Alpha=@n50.R
   ) {materialFunction=&3ColorBlend}.Result
  )
  n56 = LinearInterpolate(
   A=Multiply(A=@n53, B=VectorParameter() {group=Effect Sheen, parameterName=ESheen_Speculal_Color, defaultValue=<rgba 1 1 1 1>}.RGB)
   B=Multiply(A=@n52, B=Multiply(A=@n39, B=@n55))
   Alpha=ScalarParameter() {group=Effect Sheen, parameterName=ESheen_Speculal_Color 3Colors}
  )
  n57 = ScalarParameter() {group=Effect Sheen, parameterName="ESheen_Affect BaseColor  =1"}
  n58 = NamedRerouteUsage()
  n59 = OneMinus(None=Min(A=@n53, B=ScalarParameter() {group=Effect Sheen, parameterName=ESheen_Darkeen Emissive}))
  n60 = ScalarParameter() {group=Effect Sheen, parameterName="ESheen_Affect Emissve  =1"}
  n61 = SetMaterialAttributes!(
   MaterialAttributes=NamedRerouteUsage()
   BaseColor=Add(A=Multiply(A=@n35, B=@n54), B=LinearInterpolate(B=@n56, Alpha=@n57))
   Roughness
   Specular
   Metallic
   Normal
   EmissiveColor=Add(A=Multiply(A=@n58, B=@n59), B=LinearInterpolate(B=@n56, Alpha=@n60))
  ) {attributeSetTypes [
    #2
    #3
    #6
    #5
    #1
    #4
   ]}
  NamedRerouteDeclaration(None=@n61) {name=Mat_Base, nodeColor=<rgba 0 0.3906 1 1>}
  n62 = TextureObjectParameter() {
   group=TopSide_V02
   parameterName=TopS_00A_Sharerd_Tex_Msk
   samplerType=SAMPLERTYPE_Masks
   sortPriority=249
   texture=&T_BlackMask
  }
  NamedRerouteDeclaration(None=@n62) {name=Tex_00A_Shared_Tex, nodeColor=<rgba 1 0.8906 0 1>}
  NamedRerouteDeclaration(None=@n16.BaseColor) {name=Root_BaseColor, nodeColor=<rgba 0.4062 0 1 1>}
  NamedRerouteDeclaration(None=@n16.Roughness) {name=Root_Roughness, nodeColor=<rgba 0.4062 0 1 1>}
  NamedRerouteDeclaration(None=@n16.EmissiveColor) {name=Root_Emissive, nodeColor=<rgba 0.4062 0 1 1>}
  NamedRerouteDeclaration(None=FunctionInput(Preview=@n19.RGBA) {bUsePreviewValueAsDefault=true, inputName=IsState, inputType=FunctionInput_Vector4, sortPriority=4}) {name=IsState, nodeColor=<rgba 0.1497 1 0.1633 1>}
  NamedRerouteDeclaration(None=Multiply(A=ComponentMask(None=@n16.Normal) {g=true, r=true}, B=ScalarParameter() {group=Effect Sheen, parameterName=ESheen_Normal Warp})) {name=Root_Normal, nodeColor=<rgba 0.4062 0 1 1>}
 ]}
