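# Material function for weapon-skin wraps that restyles incoming material attributes into a comic/anime look, gated by the Effect_Stylise static switch.
# Banded cel shading from a custom light direction combines with stippling, grunge, stop-motion paper detail, Phong/Blinn highlights and metallic gradients.
# Procedural dots and hatching use projected rim vectors and also feed separate rim-light outputs.
# Outlines support time-triggered color fades, with strength selected by item-type state inputs (weapon, vehicle, character, etc.).
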MaterialFunction /Game/Weapons/WeaponSkins/Wraps/MaterialFunctions/MF_Effect_Stylise
graph {nodes [
  n1 = MaterialFunctionCall() {materialFunction=/Game/Global/Materials/Character_Parent_2023/Anime_Parent_2023/MaterialFunctions/MFC_AnimeShader_LightDirection}
  n2 = ScalarParameter() {
   desc="LockedZ only for the specular highlight.\r\nfor a streched highlight in on the Z axis use the Blinn highlight"
   group=EffectStylise_ShakeCrunch_DotsRimHatching
   parameterName=EfStylise_SpecularHighlight_LockedZVector
   sliderMax=1
  }
  n3 = AppendVector(A=Constant2Vector() {g=1, r=1}, B=Saturate(None=OneMinus(None=@n2)))
  n4 = Multiply(A=Add(A=CameraVectorWS(), B=@n1.Direction Vector), B=@n3)
  n5 = NamedRerouteUsage()
  n6 = Multiply(A=@n1.Direction Vector, B=@n3)
  n7 = StaticSwitchParameter(
   True=DotProduct(A=Normalize(VectorInput=@n4), B=@n5) {desc=Blinn}
   False=DotProduct(A=Normalize(VectorInput=@n6), B=ReflectionVectorWS(CustomWorldNormal=@n5)) {desc=Phong}
  ) {
   desc="Default is the Phong highlight.\r\nUse Blinn for a vertical streched highlight"
   group=EffectStylise_ShakeCrunch_DotsRimHatching
   parameterName=EfStylise_UseBlinnHighlight
   sortPriority=0
  }
  n8 = Saturate(None=LinearInterpolate(
    A=@n7
    B=NamedRerouteUsage()
    Alpha=ScalarParameter() {group=EffectStylise_ShakeCrunch_DotsRimHatching, parameterName=EfStylise_SpecularHighlight_CameraFacing_Blend, sliderMax=1}
   ))
  n9 = Constant() {r=1}
  n10 = ScalarParameter() {
   defaultValue=0.15
   group=EffectStylise_ShakeCrunch_DotsRimHatching
   parameterName=EfStylise_SpecularHighlight_Size
   sliderMax=1
   sliderMin=0.0001
  }
  n11 = ScalarParameter() {group=EffectStylise_ShakeCrunch_DotsRimHatching, parameterName=EfStylise_SpecularHighlight_Hardness, sliderMax=1}
  n12 = VectorParameter() {
   desc="RGB=Color, A=Opacity"
   group=EffectStylise_ShakeCrunch_DotsRimHatching
   parameterName=EfStylise_Specular_Color
   defaultValue=<rgba 1 1 1 0.25>
  }
  n13 = Multiply(A=Saturate(None=SphereMask(A=@n8, B=@n9, Radius=@n10, Hardness=@n11) {hardnessPercent=78}), B=Multiply(A=@n12.RGB, B=@n12.RGB))
  n14 = ScalarParameter() {
   defaultValue=0.3088
   group=EffectStylise_ShakeCrunch_DotsRimHatching
   parameterName=EfStylise_SpecularHighlight_Size_Stipping
   sliderMax=1
   sliderMin=0.0001
  }
  n15 = SphereMask(
   A=@n8
   B=Constant() {r=1}
   Radius=@n14
   Hardness=ScalarParameter() {group=EffectStylise_ShakeCrunch_DotsRimHatching, parameterName=EfStylise_SpecularHighlight_Hardness_Stipping, sliderMax=1}
  ) {hardnessPercent=78}
  n16 = MaterialFunctionCall!(Base=@n15, Add=NamedRerouteUsage()) {materialFunction=/Engine/Functions/Engine_MaterialFunctions02/Math/MultiplyAdd}
  n17 = VectorParameter() {
   desc="RGB=Color, A=Opacity"
   group=EffectStylise_Stippling
   parameterName=EfStylise_Specular_Color_Stipping
   defaultValue=<rgba 1 0.7604 0.6734 0>
  }
  n18 = Multiply(
   A=SphereMask(
    A=StaticSwitchParameter(True=@n16.Result, False=@n15) {group=EffectStylise_Stippling, parameterName=EfStylise_Stippling_Tex_Msk, sortPriority=0}
    B=ScalarParameter() {defaultValue=1, group=EffectStylise_Stippling, parameterName=EfStylise_Thres}
    Radius=ScalarParameter() {defaultValue=0.6435, group=EffectStylise_Stippling, parameterName=EfStylise_Radius, sliderMax=1}
    Hardness=ScalarParameter() {defaultValue=0.2149, group=EffectStylise_Stippling, parameterName=EfStylise_Hardness}
   )
   B=Multiply(A=@n17.RGB, B=@n17.RGB)
  )
  Add(A=@n13, B=@n18)
  n19 = NamedRerouteUsage()
  n20 = ScalarParameter() {defaultValue=2, group=EffectStylise_Stippling, parameterName=EfStylise_Stippling_UVScale}
  ComponentMask(None=MaterialFunctionCall!(
    ColorTex=@n19
    NormalTex
    Pos
    Nrm
    ColorUVScale=@n20
    ColorBlendFactor
    NormalStrength
    NormalUVScale
    NormalBlendFactor
    Local Space Inputs
   ) {materialFunction=/Game/Global/MaterialFunctions/MF_BiPlanar}.Color) {r=true}
  Constant() {r=0.5}
  n21 = NamedRerouteUsage()
  n22 = Saturate(None=NamedRerouteUsage())
  n23 = Constant()
  n24 = SetMaterialAttributes!(
   MaterialAttributes=@n21
   EmissiveColor=Multiply(A=@n22, B=ScalarParameter() {defaultValue=0.5, group=EffectStylise, parameterName=EfStylise_Emissive_Scale})
   BaseColor=Multiply(A=@n22, B=ScalarParameter() {defaultValue=1, group=EffectStylise, parameterName=EfStylise_BaseColor_Scale})
   Specular=@n23
   Metallic=@n23
   Roughness=Constant() {r=1}
  ) {attributeSetTypes [
    #1
    #2
    #3
    #4
    #5
   ]}
  n25 = ScalarParameter() {defaultValue=0.25, group=EffectStylise_ShakeCrunch_DotsRimHatching, parameterName=EfStylise_CelShading_Blend, sliderMax=1, sortPriority=1}
  n26 = GetMaterialAttributes(None=BlendMaterialAttributes(
    A=SetMaterialAttributes!(MaterialAttributes=@n21, BaseColor=@n22, Metallic, Specular) {attributeSetTypes [#2, #4, #3]}
    B=@n24
    Alpha=@n25
   )) {attributeGetTypes [#1, #2]}
  n27 = NamedRerouteUsage()
  n28 = Add(A=LinearInterpolate(A=@n26.EmissiveColor, B=NamedRerouteUsage(), Alpha=NamedRerouteUsage()) {constB=0}, B=NamedRerouteUsage())
  n29 = LinearInterpolate(A=@n26.BaseColor, B=NamedRerouteUsage(), Alpha=NamedRerouteUsage()) {constB=0}
  FunctionOutput(None=StaticSwitchParameter(
    True=SetMaterialAttributes!(MaterialAttributes=@n26.MaterialAttributes, Normal=@n27, EmissiveColor=@n28, BaseColor=@n29) {attributeSetTypes [#6, #1, #2]}
    False=NamedRerouteUsage()
   ) {group=Customization, parameterName=Effect_Stylise}) {outputName=MA_Out}
  n30 = NamedRerouteUsage()
  n31 = NamedRerouteUsage()
  n32 = NamedRerouteUsage()
  n33 = NamedRerouteUsage()
  FunctionOutput(None=Max(A=Max(A=@n30, B=@n31), B=Max(A=@n32, B=@n33))) {outputName=RimLight_Gradient, sortPriority=2}
  n34 = Saturate(None=Power(Base=NamedRerouteUsage()) {constExponent=2.2})
  n35 = Constant() {desc=Metallic Band Amount, r=3}
  n36 = ScalarParameter() {defaultValue=0.75, group=EffectStylise_ShakeCrunch_DotsRimHatching, parameterName=EfStylise_Metallic_Roughness, sliderMax=1}
  n37 = Saturate(None=@n36)
  n38 = Divide(
   A=MaterialFunctionCall!(In=Multiply(A=@n34, B=@n35), Ramp Width=Multiply(A=@n37) {constB=2}) {materialFunction=/Engine/Functions/Engine_MaterialFunctions01/Math/SmoothCeil}.Result
   B=@n35
  )
  n39 = ScalarParameter() {defaultValue=1, group=EffectStylise_ShakeCrunch_DotsRimHatching, parameterName=EfStylise_Metallic_Intensity, sliderMax=1}
  n40 = ScalarParameter() {
   defaultValue=-0.9
   desc="The Bias will adjust the Metallic gradient.\r\n0.0 = the center matches the Color\r\n1.0 = the inner ring matches the Color\r\n2.0 = the outer ring match the Color"
   group=EffectStylise_ShakeCrunch_DotsRimHatching
   parameterName=EfStylise_Metallic_Bias
   sliderMax=2
   sliderMin=-1
  }
  n41 = Add(A=Saturate(None=LinearInterpolate(B=@n38, Alpha=@n39) {constA=1}), B=Multiply(A=@n40) {constB=0.3333})
  LinearInterpolate(Alpha=@n41)
  MaterialFunctionCall!(Position, UVRadialScale) {materialFunction=/Game/Characters/Player/Common/MaterialFunctions/Superlevel/CharacterStabilizedScreenSpaceUVs}
  n42 = NamedRerouteUsage()
  n43 = LocalPosition() {localOrigin=InstancePreSkinning}
  MaterialFunctionCall(
   ColorTex=@n42
   Pos=Add(A=VertexInterpolator(VS=@n43.XYZ).PS, B=Constant3Vector() {constant=<rgba 0 0 14.6837 1>})
   ColorUVScale=Constant() {r=35}
  ) {materialFunction=&MF_BiPlanar}
  Max(A=NamedRerouteUsage(), B=@n30)
  Multiply(A=Max(A=@n13, B=@n18), B=NamedRerouteUsage())
  n44 = ScalarParameter() {
   defaultValue=-0.8
   group=EffectStylise_ShakeCrunch_DotsRimHatching
   parameterName=EfStylise_HighlightDotsOffset
   sliderMax=1
   sliderMin=-1
   sortPriority=20
  }
  n45 = ScalarParameter() {
   defaultValue=1.5
   group=EffectStylise_ShakeCrunch_DotsRimHatching
   parameterName=EfStylise_HighlightDotsTransition
   sliderMax=2
   sortPriority=20
  }
  n46 = Multiply(A=Add(A=NamedRerouteUsage(), B=@n44) {constB=-0.25}, B=@n45) {constB=1.5}
  n47 = NamedRerouteUsage()
  n48 = SmoothStep(Min=Constant(), Max=Constant() {r=0.1}, Value=Saturate(None=Subtract(A=@n46, B=@n47)))
  Multiply(A=@n48, B=OneMinus(None=NamedRerouteUsage()))
  Multiply(A=@n15)
  n49 = VectorParameter() {group=EffectStylise_Stippling, parameterName=EfStylise_Stippling_UVScale2, defaultValue=<rgba 16 16 0 1>}
  Multiply(A=TextureCoordinate(), B=ComponentMask(None=@n49.RGB) {g=true, r=true})
  n50 = PreSkinnedLocalBounds()
  n51 = Subtract(A=VertexInterpolator(VS=PreSkinnedPosition()).PS, B=@n50.Min)
  n52 = MaterialFunctionCall(Color=Subtract(A=@n50.Max, B=@n50.Max)) {materialFunction=&SplitComponents}
  n53 = Max(A=Max(A=@n52.R, B=@n52.R), B=@n52.B)
  NamedRerouteDeclaration(None=FunctionInput(Preview=Divide(A=@n51, B=@n53)) {bUsePreviewValueAsDefault=true, inputName=Calcated_Bounds, sortPriority=11}) {name=Caclcated_Bounds, nodeColor=<rgba 0 0.9297 1 1>}
  NamedRerouteDeclaration(None=FunctionInput(Preview=Constant4Vector().RGBA) {bUsePreviewValueAsDefault=true, inputName=IsState, inputType=FunctionInput_Vector4, sortPriority=4}) {name=IsState, nodeColor=<rgba 0.1497 1 0.1633 1>}
  NamedRerouteDeclaration(None=FunctionInput(Preview=Constant2Vector()) {bUsePreviewValueAsDefault=true, inputName=IsState_B, inputType=FunctionInput_Vector2, sortPriority=4}) {name=IsState_B, nodeColor=<rgba 0.1497 1 0.1633 1>}
  n54 = MaterialFunctionCall(Float3=Abs(None=NamedRerouteUsage())) {materialFunction=&BreakOutFloat3Components}
  n55 = Constant() {r=0.5}
  n56 = NamedRerouteUsage()
  n57 = ComponentMask(None=@n56) {b=true, desc=YZ, g=true}
  n58 = ComponentMask(None=@n56) {desc=YX, g=true, r=true}
  n59 = ComponentMask(None=@n56) {b=true, desc=XZ, r=true}
  n60 = If(A=@n54.R, B=@n55, "A > B"=@n57, "A < B"=If(A=@n54.B, B=@n55, "A > B"=@n58, "A < B"=@n59))
  n61 = Constant() {r=0.5}
  n62 = Constant() {r=0.5}
  n63 = MaterialFunctionCall(Float4=NamedRerouteUsage()) {materialFunction=&BreakOutFloat4Components}
  n64 = LinearInterpolate(
   A=If(A=@n54.B, B=@n61, "A > B"=@n58, "A < B"=If(A=@n54.R, B=@n61, "A > B"=@n57, "A < B"=@n59))
   B=If(A=@n54.G, B=@n62, "A > B"=@n59, "A < B"=If(A=@n54.B, B=@n62, "A > B"=@n58, "A < B"=@n57))
   Alpha=@n63.G
  )
  n65 = Constant() {r=0.5}
  n66 = If(A=@n54.R, B=@n65, "A > B"=@n57, "A < B"=@n58)
  NamedRerouteDeclaration(None=LinearInterpolate(A=LinearInterpolate(A=@n60, B=@n64, Alpha=@n63.R), B=If(A=@n54.G, B=@n65, "A > B"=@n59, "A < B"=@n66), Alpha=@n63.B)) {name=UVS_Cheap_Tri_1, nodeColor=<rgba 1 0 0.8438 1>}
  NamedRerouteDeclaration(None=@n13) {name=Specular_Curve, nodeColor=<rgba 1 0.4453 0 1>}
  n67 = VectorParameter() {group=EffectStylise_ShakeCrunch_DotsRimHatching, parameterName=EfStylise_RimLighVector, sortPriority=10, defaultValue=<rgba -1 0 1 1>}
  n68 = MaterialFunctionCall!(Vector=@n67.RGB) {materialFunction=/Game/Packages/Fortress_SharedMaterials/Anime_Master_Material/MF_ProjectToScreenSpace}
  n69 = NamedRerouteUsage()
  n70 = ScalarParameter() {group=EffectStylise_ShakeCrunch_DotsRimHatching, parameterName=EfStylise_RimDotsThreshold, sliderMax=2, sortPriority=20}
  n71 = Subtract(A=NamedRerouteDeclaration(None=DotProduct(A=@n68.Result, B=@n69)) {name=RimGradient, nodeColor=<rgba 0 0.3906 1 1>}, B=@n70)
  n72 = Add(A=@n71) {constB=-0.25}
  n73 = PreSkinnedNormal()
  n74 = DotProduct(A=ShaderStageSwitch(PixelShader=VertexInterpolator(VS=@n73).PS, VertexShader=@n73), B=Constant3Vector() {constant=<rgba 0 1 0 0>})
  n75 = Multiply(A=@n74, B=@n74)
  n76 = NamedRerouteDeclaration(None=Multiply(A=@n75, B=@n75) {desc=reduce rim light on the Y axis}) {name=YMask, nodeColor=<rgba 0 1 0.5078 1>}
  n77 = NamedRerouteDeclaration(None=SmoothStep(Min=Constant(), Max=Constant() {r=0.05}, Value=LinearInterpolate(A=@n71, B=@n72, Alpha=@n76) {constA=1, constB=16})) {name=AccentRimMask, nodeColor=<rgba 0.6631 0.3995 0.9131 1>}
  n78 = VectorParameter() {
   desc="RGB = Color\r\nA = Alpha / Instensity"
   group=EffectStylise_ShakeCrunch_DotsRimHatching
   parameterName=EfStylise_RimColor
   sortPriority=10
   defaultValue=<rgba 0 0.4324 0.8 1>
  }
  NamedRerouteDeclaration(None=AppendVector(A=Multiply(A=@n77, B=@n78.RGB), B=@n78.A)) {name=AccentRim, nodeColor=<rgba 0.5132 0 1 1>}
  n79 = VectorParameter() {group=EffectStylise_ShakeCrunch_DotsRimHatching, parameterName=EfStylise_RimDotsVector, sortPriority=10, defaultValue=<rgba -1 0 0.5 1>}
  n80 = DotProduct(
   A=MaterialFunctionCall(Vector=@n79.RGB) {materialFunction=&MF_ProjectToScreenSpace}.Result
   B=NamedRerouteUsage()
  )
  n81 = ScalarParameter() {
   defaultValue=-0.4
   group=EffectStylise_ShakeCrunch_DotsRimHatching
   parameterName=EfStylise_RimDotsOffset
   sliderMax=0.5
   sliderMin=-0.5
   sortPriority=20
  }
  n82 = ScalarParameter() {defaultValue=1.5, group=EffectStylise_ShakeCrunch_DotsRimHatching, parameterName=EfStylise_RimDotsTransition, sliderMax=2, sortPriority=20}
  n83 = Subtract(A=Multiply(A=Add(A=@n80, B=@n81) {constB=-0.25}, B=@n82) {constB=1.5}, B=@n70)
  n84 = Subtract(A=LinearInterpolate(A=@n83, B=Add(A=@n83) {constB=-0.25}, Alpha=NamedRerouteUsage()) {constA=1, constB=16}, B=NamedRerouteUsage())
  n85 = NamedRerouteDeclaration(None=SmoothStep(Min=Constant(), Max=Constant() {r=0.1}, Value=Saturate(None=@n84))) {name=RimDotsMask, nodeColor=<rgba 0.4766 1 0 1>}
  n86 = VectorParameter() {
   desc="RGB = Color\r\nA = Alpha / Instensity"
   group=EffectStylise_ShakeCrunch_DotsRimHatching
   parameterName=EfStylise_RimDotsColor
   sortPriority=10
   defaultValue=<rgba 0.9 0.9 0.9 0.1>
  }
  n87 = NamedRerouteUsage()
  NamedRerouteDeclaration(None=AppendVector(A=Multiply(A=@n85, B=@n86.RGB), B=Multiply(A=@n86.A, B=@n87))) {name=RimDots, nodeColor=<rgba 1 0 0.6797 1>}
  n88 = Multiply(A=NamedRerouteUsage()) {constB=-1}
  n89 = ScalarParameter() {
   defaultValue=-0.5
   group=EffectStylise_ShakeCrunch_DotsRimHatching
   parameterName=EfStylise_ShadowHatchingOffset
   sliderMax=0.5
   sliderMin=-0.5
   sortPriority=20
  }
  n90 = Subtract(A=Multiply(A=Add(A=@n88, B=@n89) {constB=-0.25}) {constB=0.5, desc=:/}, B=NamedRerouteUsage())
  n91 = NamedRerouteDeclaration(None=SmoothStep(Min=Constant(), Max=Constant() {r=0.1}, Value=Saturate(None=@n90))) {name=ShadowHatchingMask, nodeColor=<rgba 0.3594 0.3594 0.3594 1>}
  n92 = VectorParameter() {
   desc="RGB = Color\r\nA = Alpha / Instensity"
   group=EffectStylise_ShakeCrunch_DotsRimHatching
   parameterName=EfStylise_ShadowHatchingColor
   sortPriority=10
   defaultValue=<rgba 0.04 0.04 0.04 0.2>
  }
  NamedRerouteDeclaration(None=AppendVector(A=Multiply(A=@n91, B=@n92.RGB), B=@n92.A)) {name=ShadowHatching, nodeColor=<rgba 0.099 0.099 0.099 1>}
  NamedRerouteDeclaration(None=MaterialFunctionCall(
    A=VectorParameter() {parameterName=EfStylise_Metallic_ColorA, defaultValue=<rgba 0.113 0.0055 0 1>}.RGB
    B=VectorParameter() {parameterName=EfStylise_Metallic_Color_B, defaultValue=<rgba 0.5583 0.5217 0.1262 1>}.RGB
    C=VectorParameter() {parameterName=EfStylise_Metallic_ColorC, defaultValue=<rgba 1 0.9702 0.887 1>}.RGB
    Alpha=@n41
   ) {materialFunction=&3ColorBlend}.Result) {name=Metallic_Gradient, nodeColor=<rgba 0.1107 1 0.7992 1>}
  NamedRerouteDeclaration(None=MaterialFunctionCall() {materialFunction=/Game/Global/Materials/Character_Parent_2023/Anime_Parent_2023/MaterialFunctions/MFC_ActorAlighnedScreenPosition}.UVs) {name=ObjectAlighedScreenPosition, nodeColor=<rgba 1 0.867 0 1>}
  n93 = PixelNormalWS()
  NamedRerouteDeclaration(None=DotProduct(A=NamedRerouteDeclaration(None=@n93) {name=NormalWS, nodeColor=<rgba 0.0078 0 1 1>}, B=CameraVectorWS())) {name=NdotV, nodeColor=<rgba 1 0.6242 0 1>}
  n94 = MaterialFunctionCall() {materialFunction=&MFC_AnimeShader_LightDirection}
  n95 = PixelNormalWS()
  n96 = Multiply(A=NamedRerouteDeclaration(None=DotProduct(A=@n94.Direction Vector, B=@n95)) {name=LdotN, nodeColor=<rgba 0.2656 0 1 1>}) {constB=0.5}
  NamedRerouteDeclaration(None=Saturate(None=Add(A=@n96) {constB=0.5})) {name=Lambert, nodeColor=<rgba 0 0.9297 1 1>}
  NamedRerouteDeclaration(None=MaterialFunctionCall!(Fade Offset=Constant() {r=50}, Fade Distance=Constant() {r=400}) {materialFunction=/Game/Global/Materials/Character_Parent_2023/Anime_Parent_2023/MaterialFunctions/MFC_DistanceFade_FOV_compensation}.Result) {name=DistanceBlending, nodeColor=<rgba 0.4766 1 0 1>}
  n97 = ScalarParameter() {defaultValue=150, group=EffectStylise_ShakeCrunch_DotsRimHatching, parameterName=EfStylise_HatchingAmount, sliderMax=512, sortPriority=20}
  n98 = Multiply(A=NamedRerouteUsage(), B=LinearInterpolate(A=@n97, B=Multiply(A=@n97) {constB=0.5}, Alpha=NamedRerouteUsage()))
  n99 = Constant() {r=1}
  n100 = Frac(None=ComponentMask(None=Rotator(Coordinate=@n98, Time=@n99) {speed=0.75}) {g=true})
  NamedRerouteDeclaration(None=Abs(None=Add(A=@n100) {constB=-0.5})) {name=HatchingGradient, nodeColor=<rgba 0.3042 0 1 1>}
  n101 = VectorParameter() {
   group=EffectStylise_ShakeCrunch_DotsRimHatching
   parameterName=EfStylise_RimHatchingVector
   sortPriority=10
   defaultValue=<rgba 0.8229 0.15 0.4865 1>
  }
  n102 = DotProduct(
   A=MaterialFunctionCall(Vector=@n101.RGB) {materialFunction=&MF_ProjectToScreenSpace}.Result
   B=NamedRerouteUsage()
  )
  n103 = ScalarParameter() {
   defaultValue=-0.3
   group=EffectStylise_ShakeCrunch_DotsRimHatching
   parameterName=EfStylise_HatchingOffset
   sliderMax=0.5
   sliderMin=-0.5
   sortPriority=20
  }
  n104 = Subtract(A=Multiply(A=Add(A=@n102, B=@n103) {constB=-0.25}) {constB=0.5, desc=:/}, B=NamedRerouteUsage())
  n105 = NamedRerouteDeclaration(None=SmoothStep(Min=Constant(), Max=Constant() {r=0.1}, Value=Saturate(None=@n104))) {name=RimHatchingMask, nodeColor=<rgba 0.6174 0.2284 0.6354 1>}
  n106 = VectorParameter() {
   desc="RGB = Color\r\nA = Alpha / Instensity"
   group=EffectStylise_ShakeCrunch_DotsRimHatching
   parameterName=EfStylise_RimHatchingColor
   sortPriority=10
   defaultValue=<rgba 0.04 0.04 0.04 0.2>
  }
  NamedRerouteDeclaration(None=AppendVector(A=Multiply(A=@n105, B=@n106.RGB), B=@n106.A)) {name=RimHatching, nodeColor=<rgba 0.651 0 0.589 1>}
  n107 = SetMaterialAttributes!(
   MaterialAttributes
   BaseColor=TextureSample() {texture=/Game/Weapons/FORT_Rifles/T_SCAR_D}.RGB
   Normal=TextureSample() {samplerType=SAMPLERTYPE_Normal, texture=/Game/Weapons/FORT_Rifles/T_SCAR_N}.RGB
  ) {attributeSetTypes [#2, #6]}
  n108 = GetMaterialAttributes(None=NamedRerouteDeclaration(None=FunctionInput(Preview=@n107) {inputName=MA_In, inputType=FunctionInput_MaterialAttributes}) {name=MA_In, nodeColor=<rgba 1 0.9609 0 1>}) {attributeGetTypes [
    #2
    #6
    #4
    #3
    #1
   ]}
  NamedRerouteDeclaration(None=@n108.EmissiveColor) {name=Emissive_In, nodeColor=<rgba 1 0.9289 0 1>}
  n109 = NamedRerouteUsage()
  n110 = NamedRerouteUsage()
  n111 = ScalarParameter() {defaultValue=1, group=EffectStylise_ShakeCrunch_DotsRimHatching, parameterName=EfStylise_BaseColor_Desaturate}
  n112 = Multiply(
   A=NamedRerouteUsage()
   B=LinearInterpolate(
    A=ScalarParameter() {defaultValue=0.025, group=EffectStylise_ShakeCrunch_DotsRimHatching, parameterName=EfStylise_GrungeMapIntensity}
    B=ScalarParameter() {defaultValue=-0.3, group=EffectStylise_ShakeCrunch_DotsRimHatching, parameterName=EfStylise_GrungeMapIntensity_Skin}
    Alpha=Desaturation(None=@n110, Fraction=@n111)
   )
  )
  n113 = StaticSwitchParameter(True=Saturate(None=Subtract(A=@n109, B=@n112)), False=@n110) {group=EffectStylise_ShakeCrunch_DotsRimHatching, parameterName=EfStylise_Grundge_Tex_Msk, sortPriority=50}
  n114 = LinearInterpolate(
   A=Multiply(A=@n113, B=ScalarParameter() {defaultValue=0.675, group=EffectStylise, parameterName=EfStylise_ShadedDarkening}) {constB=0.66}
   B=@n113
   Alpha=NamedRerouteUsage()
  )
  n115 = Multiply(
   A=@n114
   B=Multiply(A=NamedRerouteUsage(), B=ScalarParameter() {defaultValue=1, group=EffectStylise, parameterName=EfStylise_Metalic_Strength, sliderMax=2})
  )
  n116 = NamedRerouteUsage()
  n117 = ScalarParameter() {defaultValue=0.05, group=EffectStylise_ShakeCrunch_DotsRimHatching, parameterName=EfStylise_Specular_Intensity, sliderMax=4}
  n118 = NamedRerouteUsage()
  n119 = Multiply(A=@n118, B=@n117)
  n120 = ScalarParameter() {defaultValue=1, group=EffectStylise_ShakeCrunch_DotsRimHatching, parameterName=EfStylise_Metallic_SpecularIntensity, sliderMax=4}
  n121 = Multiply(A=@n118, B=@n120)
  n122 = NamedRerouteUsage()
  NamedRerouteDeclaration(None=Add(A=LinearInterpolate(A=@n114, B=@n115, Alpha=@n116), B=LinearInterpolate(A=@n119, B=@n121, Alpha=@n122))) {name=ColorRaw, nodeColor=<rgba 0 0.5781 1 1>}
  n123 = ScalarParameter() {defaultValue=150, group=EffectStylise_ShakeCrunch_DotsRimHatching, parameterName=EfStylise_DotsAmount, sliderMax=512, sortPriority=20}
  n124 = Multiply(A=NamedRerouteUsage(), B=LinearInterpolate(A=@n123, B=Multiply(A=@n123) {constB=0.5}, Alpha=NamedRerouteUsage()))
  n125 = Constant() {r=1}
  n126 = Add(A=Frac(None=Rotator(Coordinate=@n124, Time=@n125) {speed=0.75})) {constB=-0.5}
  n127 = Constant()
  NamedRerouteDeclaration(None=Add(A=Distance(A=@n126, B=@n127)) {constB=-0.1, desc=to get full transition}) {name=DotsGradient, nodeColor=<rgba 1 0.0938 0 1>}
  n128 = NamedRerouteDeclaration(None=@n48) {name=HighlightDotsMask, nodeColor=<rgba 0.4766 1 0 1>}
  n129 = VectorParameter() {
   desc="RGB = Color\r\nA = Alpha / Instensity"
   group=EffectStylise_ShakeCrunch_DotsRimHatching
   parameterName=EfStylise_HighlightDotsColor
   sortPriority=10
   defaultValue=<rgba 0.9 0.9 0.9 0.1>
  }
  n130 = NamedRerouteUsage()
  NamedRerouteDeclaration(None=AppendVector(A=Multiply(A=@n128, B=@n129.RGB), B=Multiply(A=@n129.A, B=@n130))) {name=HighlightDots, nodeColor=<rgba 0.9242 1 0.9798 1>}
  n131 = TextureSample() {samplerType=SAMPLERTYPE_Masks, texture=/Game/Weapons/FORT_Rifles/Textures/T_Scar_CM}
  n132 = MaterialFunctionCall(Float4=FunctionInput(Preview=@n131.RGBA) {inputName=CM_Msk, inputType=FunctionInput_Vector4, sortPriority=4}) {materialFunction=&BreakOutFloat4Components}
  NamedRerouteDeclaration(None=@n132.G) {name=Msk_Curvature, nodeColor=<rgba 0 1 0.7422 1>}
  NamedRerouteDeclaration(None=@n108.Metallic) {name=Metallic_In, nodeColor=<rgba 0 0.9581 1 1>}
  NamedRerouteDeclaration(None=@n108.Specular) {name=Specular_In, nodeColor=<rgba 1 0.9289 0 1>}
  NamedRerouteDeclaration(None=@n132.B) {name=Msk_Accent, nodeColor=<rgba 0 1 0.7656 1>}
  n133 = ComponentMask(None=ViewProperty() {property=MEVP_ResolutionFraction}.Property) {r=true}
  NamedRerouteDeclaration(None=Saturate(None=InverseLinearInterpolate(Value=@n133) {constA=0.3})) {name=ResScale, nodeColor=<rgba 0 0.6484 1 1>}
  n134 = NamedRerouteUsage()
  n135 = ScalarParameter() {defaultValue=0.2, group=EffectStylise_ShakeCrunch_DotsRimHatching, parameterName=EfStylise_PaperDetail_SteppedTime_Scale, sliderMax=1}
  n136 = Multiply(A=Time(), B=@n135)
  n137 = ScalarParameter() {defaultValue=12, group=EffectStylise_ShakeCrunch_DotsRimHatching, parameterName=EfStylise_PaperDetail_SteppedTime_Steps, sliderMax=32}
  n138 = Floor(None=Multiply(A=Frac(None=@n136), B=@n137))
  n139 = ScalarParameter() {defaultValue=0.125, group=EffectStylise_ShakeCrunch_DotsRimHatching, parameterName=EfStylise_PaperDetail_SteppedTime_Offset, sliderMax=1}
  n140 = ScalarParameter() {defaultValue=0.2, group=EffectStylise_ShakeCrunch_DotsRimHatching, parameterName=EfStylise_PaperDetail_Tiling, sliderMax=2}
  n141 = Multiply(A=Add(A=NamedRerouteUsage(), B=Multiply(A=@n138, B=@n139)), B=@n140)
  n142 = ChannelMaskParameter(None=TextureSample(UVs=Frac(None=@n141), Tex=NamedRerouteUsage()) {samplerType=SAMPLERTYPE_Masks}.RGB) {group=EffectStylise_ShakeCrunch_DotsRimHatching, maskChannel=Green, parameterName=EfStylise_PaperDetail_Channel, defaultValue=<rgba 0 1 0 0>}
  n143 = ScalarParameter() {defaultValue=0.2, group=EffectStylise_ShakeCrunch_DotsRimHatching, parameterName=EfStylise_PaperDetail_Bias, sliderMax=1, sliderMin=-1}
  n144 = ScalarParameter() {defaultValue=0.8, group=EffectStylise_ShakeCrunch_DotsRimHatching, parameterName=EfStylise_PaperDetail_Exponent, sliderMax=2}
  n145 = Power(Base=Add(A=@n142.R, B=@n143), Exp=@n144) {constExponent=0.4545}
  NamedRerouteDeclaration(None=StaticSwitchParameter(
    True=MaterialFunctionCall(Base=@n134, Blend=@n145) {materialFunction=&Blend_Overlay}.Result
    False=@n134
   ) {group=EffectStylise_ShakeCrunch_DotsRimHatching, parameterName=EfStylise_PaperDetail_Tex_Msk, sortPriority=50}) {name=BaseColor_withGrunge, nodeColor=<rgba 1 0.7969 0 1>}
  NamedRerouteDeclaration(None=@n108.BaseColor) {name=BaseColor_In, nodeColor=<rgba 1 0.1346 0 1>}
  n146 = ScalarParameter() {defaultValue=0.75, group=EffectStylise_Stippling, parameterName=EfStylise_Stippling_Min, sliderMax=1}
  n147 = ScalarParameter() {defaultValue=1, group=EffectStylise_Stippling, parameterName=EfStylise_Stippling_Max}
  n148 = Add(
   A=DotProduct(
    A=MaterialFunctionCall() {materialFunction=&MFC_AnimeShader_LightDirection}.Direction Vector
    B=PixelNormalWS()
   )
   B=ScalarParameter() {group=EffectStylise_ShakeCrunch_DotsRimHatching, parameterName=EfStylise_ShadingCurve_Bias, sliderMax=1, sliderMin=-1}
  )
  n149 = ScalarParameter() {defaultValue=0.6, group=EffectStylise_ShakeCrunch_DotsRimHatching, parameterName=EfStylise_ShadingCurve_Width, sliderMax=1}
  n150 = ScalarParameter() {defaultValue=1, group=EffectStylise_ShakeCrunch_DotsRimHatching, parameterName=EfStylise_ShadingCurve_Bands, sliderMax=4, sliderMin=1}
  n151 = Saturate(None=MaterialFunctionCall!(NdL=Saturate(None=@n148), Width=@n149, Bands Amount=@n150) {materialFunction=/Game/Global/Materials/Character_Parent_2023/Anime_Parent_2023/MaterialFunctions/MF_SmoothCelShadingWithBands}.Divide)
  n152 = NamedRerouteUsage()
  n153 = ChannelMaskParameter(None=TextureSample(UVs=Multiply(A=@n152, B=@n20), Tex=@n19).RGBA) {group=EffectStylise_Stippling, maskChannel=Alpha, parameterName=EfStylise_Stippling_Msk_Channel, defaultValue=<rgba 0 0 0 1>}
  n154 = StaticSwitchParameter(
   True=MaterialFunctionCall(Base=@n151, Add=NamedRerouteDeclaration(None=@n153.R) {name=StippingMask, nodeColor=<rgba 0.2891 0 1 1>}) {materialFunction=&MultiplyAdd}.Result
   False=@n151
  ) {defaultValue=true, group=EffectStylise_Stippling, parameterName=EfStylise_Stippling_Tex_Msk, sortPriority=0}
  NamedRerouteDeclaration(None=Saturate(None=SmoothStep(Min=@n146, Max=@n147, Value=@n154))) {name=Shading_Gradient, nodeColor=<rgba 1 0.1875 0 1>}
  n155 = PreSkinnedNormal()
  NamedRerouteDeclaration(None=FunctionInput(Preview=VertexInterpolator(VS=@n155).PS) {bUsePreviewValueAsDefault=true, inputName=Pre-Skinned Local Normal, sortPriority=10}) {name=PreSkin_Normal, nodeColor=<rgba 0.8281 1 0 1>}
  n156 = TextureObjectParameter() {
   group=TopSide_V02
   parameterName=TopS_00A_Sharerd_Tex_Msk
   samplerType=SAMPLERTYPE_Masks
   sortPriority=249
   texture=&T_BlackMask
  }
  n157 = TextureObjectParameter() {
   group=TopSide_V02
   parameterName=TopS_00B_Sharerd_Tex_Msk
   samplerType=SAMPLERTYPE_Masks
   sortPriority=249
   texture=&T_BlackMask
  }
  n158 = NamedRerouteDeclaration(None=@n157) {name=Tex_00B_Shared_Tex, nodeColor=<rgba 1 0.8906 0 1>}
  n159 = TextureObjectParameter() {
   group=EffectStylise
   parameterName=EfStylise_Tex-Msk
   samplerType=SAMPLERTYPE_Masks
   texture=/Game/Weapons/WeaponSkins/Wraps/Textures/T_Wrap_Stylise_Msk
  }
  NamedRerouteDeclaration(None=StaticSwitchParameter(
    True=NamedRerouteDeclaration(None=@n156) {name=Tex_00A_Shared_Tex, nodeColor=<rgba 1 0.8906 0 1>}
    False=StaticSwitchParameter(True=@n158, False=@n159) {group=EffectStylise, parameterName="EfStylise_Tex--Msk-->00B_Shared_Tex"}
   ) {group=EffectStylise, parameterName="EfStylise_Tex--Msk-->00A_Shared_Tex"}) {name=Tex--Msk, nodeColor=<rgba 0 1 0.4844 1>}
  NamedRerouteDeclaration(None=@n131.RGBA) {name=CM_Mask, nodeColor=<rgba 1 0 0.3047 1>}
  NamedRerouteDeclaration(None=@n132.R) {name=Msk_Prime, nodeColor=<rgba 0 1 0.7422 1>}
  n160 = Multiply(
   A=Subtract(A=NamedRerouteUsage(), B=ScalarParameter() {defaultValue=0.5, group=EffectStylise-outline, parameterName=EfStylise_Outline_Position})
   B=NamedRerouteUsage()
  )
  n161 = ScalarParameter() {defaultValue=1, group=EffectStylise-outline, parameterName=EfStylise_Outline_Fallof}
  n162 = Saturate(None=SmoothStep(
    Min=ScalarParameter() {defaultValue=0.2, group=EffectStylise-outline, parameterName=EfStylise_Outline_Hardness_Min}
    Max=ScalarParameter() {defaultValue=0.3, group=EffectStylise-outline, parameterName=EfStylise_Outline_Hardness_Max}
    Value=Power(Base=@n160, Exp=@n161)
   ))
  NamedRerouteDeclaration(None=Min(
    A=@n162
    B=ScalarParameter() {defaultValue=1, group=EffectStylise-outline, parameterName="EfStylise_OutlineEmiss-Disable 0=off", sortPriority=49}
   )) {name=outLine, nodeColor=<rgba 1 0 0.2109 1>}
  n163 = VectorParameter() {group=EffectStylise-outline, parameterName=EfStylise_OutlineEmiss_Color_Start, defaultValue=<rgba 0 0 0 1>}
  n164 = VectorParameter() {group=EffectStylise-outline, parameterName=EfStylise_OutlineEmiss_Color_End, defaultValue=<rgba 0 0 0 1>}
  n165 = Subtract(
   A=Time()
   B=Add(
    A=ScalarParameter() {defaultValue=-9999999, group=EffectStylise-outline-Reactive, parameterName=EfStylise_TFBP-Ourline---, sortPriority=600}
    B=ScalarParameter() {group=EffectStylise-outline-Reactive, parameterName=EfStylise-Reactive-Ourline-Delay}
   )
  )
  n166 = ScalarParameter() {defaultValue=0.25, group=EffectStylise-outline-Reactive, parameterName=EfStylise-Reactive-Ourline-Duration}
  n167 = Power(
   Base=Saturate(None=Divide(A=@n165, B=@n166))
   Exp=ScalarParameter() {defaultValue=0.25, group=EffectStylise-outline-Reactive, parameterName=EfStylise-Reactive-Ourline-Falloff}
  )
  NamedRerouteDeclaration(None=LinearInterpolate(A=@n163.RGB, B=@n164.RGB, Alpha=@n167)) {name=outlineColor, nodeColor=<rgba 0 0.5312 1 1>}
  n168 = ScalarParameter() {defaultValue=10, group=EffectStylise-outline, parameterName=EfStylise_Outline_Strength_Vehicle, sortPriority=40}
  n169 = ScalarParameter() {defaultValue=2, group=EffectStylise-outline, parameterName=EfStylise_Outline_Strength_Vehicle_Side, sortPriority=40}
  n170 = MaterialFunctionCall(Float4=NamedRerouteUsage()) {materialFunction=&BreakOutFloat4Components}
  n171 = LinearInterpolate(
   A=LinearInterpolate(
    A=ScalarParameter() {defaultValue=4, group=EffectStylise-outline, parameterName=EfStylise_Outline_Strength_Weapon, sortPriority=40}
    B=LinearInterpolate(A=@n168, B=@n169, Alpha=@n170.G)
    Alpha=@n170.R
   )
   B=LinearInterpolate(
    A=ScalarParameter() {defaultValue=10, group=EffectStylise-outline, parameterName=EfStylise_Outline_Strength_Mech, sortPriority=40}
    B=ScalarParameter() {defaultValue=6, group=EffectStylise-outline, parameterName=EfStylise_Outline_Strength_Character, sortPriority=40}
    Alpha=@n170.A
   )
   Alpha=@n170.B
  )
  n172 = ScalarParameter() {defaultValue=10, group=EffectStylise-outline, parameterName=EfStylise_Outline_Strength_Backpack, sortPriority=40}
  n173 = MaterialFunctionCall(Float2=NamedRerouteUsage()) {materialFunction=&BreakOutFloat2Components}
  NamedRerouteDeclaration(None=LinearInterpolate(
    A=LinearInterpolate(A=@n171, B=@n172, Alpha=@n173.R)
    B=ScalarParameter() {defaultValue=6, group=EffectStylise-outline, parameterName=EfStylise_Outline_Strength_Pickaxe, sortPriority=40}
    Alpha=@n173.G
   )) {name=Curvature_Strength_Edge, nodeColor=<rgba 0 0.7188 1 1>}
  n174 = NamedRerouteUsage()
  n175 = NamedRerouteUsage()
  n176 = ComponentMask(None=@n175) {a=true}
  n177 = NamedRerouteUsage()
  n178 = Add(
   A=Multiply(A=ComponentMask(None=@n174) {b=true, g=true, r=true}, B=ComponentMask(None=@n174) {a=true}) {desc=Rim Dots}
   B=Multiply(A=ComponentMask(None=@n175) {b=true, g=true, r=true}, B=Multiply(A=@n176, B=@n177)) {desc=Highlight Dots}
  )
  n179 = NamedRerouteUsage()
  n180 = ComponentMask(None=@n179) {a=true}
  n181 = NamedRerouteUsage()
  n182 = NamedRerouteUsage()
  n183 = ComponentMask(None=@n182) {a=true}
  n184 = NamedRerouteUsage()
  n185 = Add(
   A=Subtract(A=Multiply(A=@n180, B=@n181), B=ComponentMask(None=@n179) {b=true, g=true, r=true}) {desc=Rim Hatching}
   B=Subtract(A=Multiply(A=@n183, B=@n184), B=ComponentMask(None=@n182) {b=true, g=true, r=true}) {desc=Shadow Hatching}
  )
  NamedRerouteDeclaration(None=FunctionOutput(None=Subtract(A=@n178, B=@n185)) {outputName=RimLight_Color, sortPriority=1}) {name=RimLightColor, nodeColor=<rgba 0.2891 1 0 1>}
  n186 = VectorParameter() {group=EffectStylise-outline, parameterName=EfStylise_OutlineBC_Color_Start, defaultValue=<rgba 0 0 0 1>}
  NamedRerouteDeclaration(None=LinearInterpolate(
    A=@n186.RGB
    B=VectorParameter() {group=EffectStylise-outline, parameterName=EfStylise_OutlineBC_Color_End, defaultValue=<rgba 0 0 0 1>}.RGB
    Alpha=@n167
   )) {name=outlineColor_1, nodeColor=<rgba 0 0.5312 1 1>}
  NamedRerouteDeclaration(None=Min(
    A=@n162
    B=ScalarParameter() {defaultValue=1, group=EffectStylise-outline, parameterName="EfStylise_OutlineBC-Disable 0=off", sortPriority=49}
   )) {name=outLine_1, nodeColor=<rgba 1 0 0.2109 1>}
  n187 = NamedRerouteUsage()
  n188 = ScalarParameter() {defaultValue=0.003, group=EffectStylise_ShakeCrunch_DotsRimHatching, parameterName=EfStylise_Grundge_UVScale}
  n189 = ChannelMaskParameter(None=TextureSample(UVs=Multiply(A=@n187, B=@n188), Tex=@n42).RGBA) {group=EffectStylise_ShakeCrunch_DotsRimHatching, maskChannel=Blue, parameterName=EfStylise_Grundge_Channel, defaultValue=<rgba 0 0 1 0>}
  NamedRerouteDeclaration(None=@n189.R) {name=Grunge, nodeColor=<rgba 0.5469 0 1 1>}
  NamedRerouteDeclaration(None=@n108.Normal) {name=Normal_In, nodeColor=<rgba 0 0.3061 1 1>}
  n190 = NamedRerouteUsage()
  NamedRerouteDeclaration(None=MaterialFunctionCall!(Color=MaterialFunctionCall!(In=@n190, LUT, LUT Multiplier) {materialFunction=/Game/Packages/Fortress_SharedMaterials/Anime_Master_Material/MF_ApplyLUT}.Result) {materialFunction=/Game/Packages/Fortress_SharedMaterials/Anime_Master_Material/MF_TODColorAdjustment}.Result) {name=ColorFinal, nodeColor=<rgba 1 0.6797 0 1>}
  n191 = Transform(None=NamedRerouteUsage())
  NamedRerouteDeclaration(None=Transform(None=Normalize(VectorInput=@n191)) {transformSourceType=TRANSFORMSOURCE_World, transformType=TRANSFORM_Tangent}) {name=Normal_Out, nodeColor=<rgba 0 1 0.7995 1>}
  NamedRerouteUsage()
  NamedRerouteUsage()
  NamedRerouteUsage()
  n192 = ScalarParameter() {defaultValue=1, group=EffectStylise-outline, parameterName="EfStylise_Outline-Disable 0=off", sortPriority=49}
  n193 = Saturate(None=LinearInterpolate(
    A=ScalarParameter() {defaultValue=10, group=EffectStylise-outline, parameterName=EfStylise_Curvature_Mask A, sortPriority=50}
    B=ScalarParameter() {defaultValue=-7, group=EffectStylise-outline, parameterName=EfStylise_Curvature_Mask B, sortPriority=50}
    Alpha=NamedRerouteUsage()
   ))
  OneMinus(None=Saturate(None=Add(A=@n192, B=@n193)))
  PixelNormalWS()
  PixelNormalWS()
  PixelNormalWS()
  ScalarParameter() {defaultValue=5, group=EffectStylise-outline, parameterName=EfStylise_Outline_Strength}
  TransformPosition(None=WorldPosition() {worldPositionShaderOffset=WPT_ExcludeAllShaderOffsets}.XYZ) {transformSourceType=TRANSFORMPOSSOURCE_World}
  Transform(None=NamedRerouteUsage())
  VertexInterpolator(VS=PreSkinnedPosition())
  VertexNormalWS()
  VertexNormalWS()
  VertexNormalWS()
 ]}
