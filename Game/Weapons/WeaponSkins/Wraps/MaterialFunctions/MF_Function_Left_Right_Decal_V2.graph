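# Material function for weapon-skin wraps that blends a switchable decal layer over the incoming material and outputs its masks.
# Left/right or side projection places the decal, with per-object-type UV scale, offset and rotation plus optional safe zones.
# Flipbook masks, PBR overrides, emissive and numeric counter stages feed successive attribute blends.
# Static switches select features, including reactive animation, time-of-day opacity and top-side layer masking; platform switches drop normals on low-end mobile.
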MaterialFunction /Game/Weapons/WeaponSkins/Wraps/MaterialFunctions/MF_Function_Left_Right_Decal_V2
graph {nodes [
  n1 = NamedRerouteUsage()
  Abs(None=@n1)
  n2 = Time()
  n3 = ScalarParameter() {group=Function_Left_Right_Decal V2-FlipBook, parameterName=FDecalV2_FlipBook_Speed}
  n4 = Saturate(None=Fmod(
    A=Multiply(A=@n2, B=@n3)
    B=ScalarParameter() {defaultValue=1, group=Function_Left_Right_Decal V2-FlipBook, parameterName=FDecalV2_FlipBook_Speed_Delay}
   ))
  n5 = Time()
  n6 = Add(
   A=ScalarParameter() {defaultValue=-9999999, group=Function_Left_Right_Decal V2-FlipBook, parameterName=FDecalV2_FlipBook_Time_FromBP}
   B=ScalarParameter() {group=Function_Left_Right_Decal V2-FlipBook, parameterName=FDecalV2_FlipBook_Time_Delay}
  )
  n7 = Saturate(None=Divide(
    A=Subtract(A=@n5, B=@n6)
    B=ScalarParameter() {defaultValue=1, group=Function_Left_Right_Decal V2-FlipBook, parameterName=FDecalV2_FlipBook_Time_Duration}
   ))
  n8 = Add(A=@n4, B=Add(A=@n7) {constB=-1})
  n9 = Frac(None=LinearInterpolate(B=ScalarParameter() {defaultValue=1, group=Function_Left_Right_Decal V2-FlipBook, parameterName=FDecalV2_FlipBook_Repeat}, Alpha=@n7) {constB=0})
  n10 = Ceil(None=Multiply(A=@n2, B=ScalarParameter() {defaultValue=3, group=Function_Left_Right_Decal V2-FlipBook, parameterName=FDecalV2_FlipBook_Speed_Random}) {constB=8})
  n11 = Constant2Vector() {g=0.2562, r=0.0136}
  n12 = VectorParameter() {
   group=Function_Left_Right_Decal V2-FlipBook
   parameterName=FDecalV2_FlipBook_Random_Clamp
   channelNames {a="", b="", g=Max, r=Min}
   defaultValue=<rgba 0 1 0 1>
  }
  n13 = Clamp(
   None=MaterialFunctionCall!(2 Vector=Add(A=@n10, B=@n11), RandomSeed) {materialFunction=&/MF_Random_2}.utput
   Min=@n12.Min
   Max=@n12.Min
  )
  n14 = StaticSwitchParameter(True=@n13, False=@n4) {group=Function_Left_Right_Decal V2-FlipBook, parameterName=FDecalV2_FlipBook_Random_Animation}
  n15 = Add(
   A=StaticSwitchParameter(
    True=Saturate(None=@n8)
    False=StaticSwitchParameter(True=@n9, False=@n14) {group=Function_Left_Right_Decal V2-FlipBook, parameterName=FDecalV2_FlipBook_Time__Reactive}
   ) {group=Function_Left_Right_Decal V2-FlipBook, parameterName=FDecalV2_FlipBook_Time__Reactive + Speed}
   B=ScalarParameter() {group=Function_Left_Right_Decal V2-FlipBook, parameterName="FDecalV2_FlipBook_Frame ##"}
  )
  n16 = ScalarParameter() {defaultValue=2, group=Function_Left_Right_Decal V2-FlipBook, parameterName=FDecalV2_FlipBook_Number_Rows}
  n17 = ScalarParameter() {defaultValue=2, group=Function_Left_Right_Decal V2-FlipBook, parameterName=FDecalV2_FlipBook_Number_Columns}
  n18 = StaticSwitchParameter(
   True=NamedRerouteUsage()
   False=StaticSwitchParameter(True=NamedRerouteUsage(), False=NamedRerouteUsage()) {desc=Look To TopSide Shared, group=Function_Left_Right_Decal V2, parameterName="FDecalV2_Mask__Tex -->TopS_00B_Sharerd_Tex_Msk"}
  ) {desc=Look To TopSide Shared, group=Function_Left_Right_Decal V2, parameterName="FDecalV2_Mask__Tex -->TopS_00A_Sharerd_Tex_Msk"}
  n19 = NamedRerouteUsage()
  n20 = LinearInterpolate(A=Saturate(None=@n19), B=@n19, Alpha=NamedRerouteUsage())
  n21 = NamedRerouteUsage()
  n22 = NamedRerouteUsage()
  n23 = TextureProperty(Tex=@n22)
  n24 = Multiply(A=@n16, B=@n17)
  n25 = MaterialFunctionCall!(UVs=@n21, Texture Size=Multiply(A=Divide(A=@n23, B=@n24), B=NamedRerouteUsage())) {materialFunction=/Engine/Functions/Engine_MaterialFunctions01/Texturing/ComputeMipLevel}
  n26 = StaticBool() {value=true}
  n27 = MaterialFunctionCall(
   "Animation  Phase (0-1)"=@n15
   Number of Rows=@n16
   Number of Columns=@n17
   Texture=@n18
   UVs=@n20
   MipBias/Level=@n25.MipLevel
   "Clamp Anim (see tooltip)"=@n26
  ) {materialFunction=&FlipBook}
  AppendVector(A=@n27.Result, B=@n27.Result)
  n28 = NamedRerouteUsage()
  n29 = NamedRerouteUsage()
  n30 = NamedRerouteUsage()
  n31 = NamedRerouteUsage()
  n32 = ScalarParameter() {group=Function_Left_Right_Decal V2, parameterName="FDecalV2_BaseColor=0-->Emissive=1", sortPriority=-5}
  n33 = LinearInterpolate(A=@n31, Alpha=@n32) {constB=0}
  n34 = NamedRerouteUsage()
  n35 = NamedRerouteUsage()
  n36 = ScalarParameter() {group=Function_Left_Right_Decal V2, parameterName=FDecalV2_Rougness_High, sortPriority=50}
  n37 = ScalarParameter() {defaultValue=1, group=Function_Left_Right_Decal V2, parameterName=FDecalV2_Rougness_Low, sortPriority=50}
  n38 = ChannelMaskParameter(None=NamedRerouteUsage()) {group=Function_Left_Right_Decal V2, parameterName=FDecalV2_Rougness Mask_Channel, sortPriority=50}
  n39 = LinearInterpolate(
   A=ScalarParameter() {group=Function_Left_Right_Decal V2, parameterName=FDecalV2_Spec_Low, sortPriority=56}
   B=ScalarParameter() {defaultValue=1, group=Function_Left_Right_Decal V2, parameterName=FDecalV2_Spec_High, sortPriority=56}
   Alpha=ChannelMaskParameter(None=NamedRerouteUsage()) {group=Function_Left_Right_Decal V2, parameterName=FDecalV2_Spec Mask_Channel, sortPriority=56}.R
  )
  n40 = LinearInterpolate(
   A=ScalarParameter() {group=Function_Left_Right_Decal V2, parameterName=FDecalV2_Metallic_Low, sortPriority=55}
   B=ScalarParameter() {defaultValue=1, group=Function_Left_Right_Decal V2, parameterName=FDecalV2_Metallic_High, sortPriority=55}
   Alpha=ChannelMaskParameter(None=NamedRerouteUsage()) {group=Function_Left_Right_Decal V2, parameterName=FDecalV2_Metallic Mask_Channel, sortPriority=55}.R
  )
  n41 = Multiply(
   A=NamedRerouteUsage()
   B=ScalarParameter() {defaultValue=1, group=Function_Left_Right_Decal V2, parameterName=FDecalV2_BaseColor_TOD_TBrightness}
  )
  n42 = LinearInterpolate(
   A=NamedRerouteUsage()
   Alpha=ScalarParameter() {defaultValue=1, group=Function_Left_Right_Decal V2, parameterName="FDecalV2_Block_Root_Emissive    1=On", sortPriority=-5}
  ) {constB=0}
  n43 = LinearInterpolate(A=@n42, B=Add(A=@n31, B=@n42), Alpha=@n32)
  n44 = SetMaterialAttributes!(
   MaterialAttributes=NamedRerouteUsage()
   BaseColor=StaticSwitch(True=LinearInterpolate(B=@n33, Alpha=@n34), False=@n33, Value=@n35)
   Roughness=StaticSwitchParameter(
    True=ScalarParameter() {defaultValue=0.5, group=Function_Left_Right_Decal V2, parameterName=FDecalV2_Rougness, sortPriority=50}
    False=LinearInterpolate(A=@n36, B=@n37, Alpha=@n38.R)
   ) {defaultValue=true, group=Function_Left_Right_Decal V2, parameterName=FDecalV2-Roughness--Simple, sortPriority=50}
   Specular=StaticSwitchParameter(
    True=ScalarParameter() {defaultValue=0.5, group=Function_Left_Right_Decal V2, parameterName=FDecalV2_Spec, sortPriority=56}
    False=Saturate(None=@n39)
   ) {defaultValue=true, group=Function_Left_Right_Decal V2, parameterName=FDecalV2-Spec--Simple, sortPriority=56}
   Metallic=StaticSwitchParameter(True=ScalarParameter() {group=Function_Left_Right_Decal V2, parameterName=FDecalV2_Metallic, sortPriority=55}, False=Saturate(None=@n40)) {defaultValue=true, group=Function_Left_Right_Decal V2, parameterName=FDecalV2-Metalic--Simple, sortPriority=55}
   Normal=NamedRerouteUsage()
   EmissiveColor=StaticSwitch(True=LinearInterpolate(A=@n41, B=@n43, Alpha=@n34), False=@n43, Value=@n35)
   CustomizedUV7
  ) {attributeSetTypes [
    #1
    #2
    #3
    #4
    #5
    #6
    #7
   ]}
  n45 = NamedRerouteUsage()
  n46 = NamedRerouteUsage()
  n47 = Multiply(A=NamedRerouteUsage(), B=Multiply(A=@n45, B=@n46))
  n48 = Multiply(
   A=@n47
   B=LinearInterpolate(
    A=ScalarParameter() {defaultValue=1, group=Function_Left_Right_Decal V2_Reactive, parameterName=FDecalV2_Opacity Time__Start, sortPriority=31}
    B=ScalarParameter() {group=Function_Left_Right_Decal V2_Reactive, parameterName=FDecalV2_Opacity Time__End, sortPriority=31}
    Alpha=NamedRerouteUsage()
   )
  )
  n49 = StaticSwitch(
   True=@n30
   False=BlendMaterialAttributes(
    A=@n30
    B=@n44
    Alpha=StaticSwitchParameter(True=@n48, False=@n47) {group=Function_Left_Right_Decal V2_Reactive, parameterName=FDecalV2_Opacity_Reactive---------------------, sortPriority=0}
   ) {vertexAttributeBlendType=UseA}
   Value=StaticSwitch(
    True=StaticBoolParameter() {group=Function_Left_Right_Decal V2-Counter, parameterName=FDecalV2_Counter_Only, sortPriority=0}
    False=StaticBool()
    Value=NamedRerouteUsage()
   )
  )
  n50 = GetMaterialAttributes(None=@n49) {attributeGetTypes [#6, #1]}
  n51 = ScalarParameter() {
   group=Function_Left_Right_Decal V2_Emissive
   parameterName="FDecalV2_Emissive_Remove_Character_Front_Msk Emissive   1=On"
   sliderMax=1
   sortPriority=0
  }
  n52 = LinearInterpolate(A=NamedRerouteUsage(), B=NamedRerouteUsage(), Alpha=Saturate(None=Ceil(None=@n51)))
  n53 = Multiply(A=NamedRerouteUsage(), B=@n52)
  n54 = Add(A=@n50.EmissiveColor, B=@n53)
  n55 = ScalarParameter() {group=Function_Left_Right_Decal V2_Emissive, parameterName="FDecalV2_Emissive------->BaseColor", sliderMax=1, sortPriority=0}
  n56 = NamedRerouteUsage()
  n57 = OneMinus(None=@n52)
  n58 = Multiply(A=@n50.BaseColor, B=Saturate(None=Add(A=@n56, B=@n57)))
  n59 = Add(A=@n58, B=@n53)
  n60 = StaticSwitchParameter(
   True=SetMaterialAttributes!(
    MaterialAttributes=@n50.MaterialAttributes
    EmissiveColor=LinearInterpolate(A=@n54, B=@n50.EmissiveColor, Alpha=@n55)
    BaseColor=LinearInterpolate(A=@n58, B=@n59, Alpha=@n55)
   ) {attributeSetTypes [#6, #1]}
   False=@n49
  ) {group=Function_Left_Right_Decal V2_Emissive, parameterName=FDecalV2_Emissive, sortPriority=0}
  n61 = GetMaterialAttributes(None=@n60) {attributeGetTypes [
    #6
    #1
    #4
    #3
    #2
    #5
   ]}
  n62 = NamedRerouteUsage()
  n63 = VectorParameter() {
   group=Function_Left_Right_Decal V2-Counter
   parameterName=FDecalV2_Counter_PBR
   channelNames {a="BC--->Em", b=Spec, g=Metalic, r=Roughness}
   defaultValue=<rgba 0.5 0 0.5 0>
  }
  n64 = LinearInterpolate(B=@n62, Alpha=@n63."BC--->Em")
  n65 = VectorParameter() {group=Function_Left_Right_Decal V2-Counter, parameterName=FDecalV2_Counter_Emissive_Color, defaultValue=<rgba 1 1 1 1>}
  n66 = Multiply(A=@n65.RGB, B=@n65.RGB)
  n67 = Multiply(
   A=@n66
   B=LinearInterpolate(
    A=ScalarParameter() {defaultValue=1, group=Function_Left_Right_Decal V2-Counter_Reactive, parameterName=FDecalV2_Counter_Brightness Start}
    B=ScalarParameter() {group=Function_Left_Right_Decal V2-Counter_Reactive, parameterName=FDecalV2_Counter_Brightness End}
    Alpha=NamedRerouteUsage()
   )
  )
  n68 = LinearInterpolate(
   A=ScalarParameter() {defaultValue=1, group=Function_Left_Right_Decal V2-Counter_Reactive, parameterName=FDecalV2_Counter_DarkenBase Start}
   B=ScalarParameter() {defaultValue=1, group=Function_Left_Right_Decal V2-Counter_Reactive, parameterName=FDecalV2_Counter_DarkenBase End}
   Alpha=NamedRerouteUsage()
  )
  n69 = LinearInterpolate(A=@n62, Alpha=@n63."BC--->Em") {constB=0}
  n70 = LinearInterpolate(
   A=@n61.Normal
   B=NamedRerouteUsage()
   Alpha=ScalarParameter() {group=Function_Left_Right_Decal V2-Counter, parameterName="FDecalV2_Counter_PBR ---->Root_Normals"}
  )
  n71 = LinearInterpolate(
   A=@n70
   B=Add(A=@n70, B=NamedRerouteUsage())
   Alpha=ScalarParameter() {defaultValue=1, group=Function_Left_Right_Decal V2-Counter, parameterName=FDecalV2_Counter_Normal_Intensty}
  )
  n72 = MaterialFunctionCall!("Xb1,PS4"=@n71, "PC-High&&Epic,XSX,PS5"=@n71, "PC-Perf,Android"=@n70, Switch=@n71, "PC-Low(NotPerf)&&Medium"=@n71) {materialFunction=/Game/Effects/Niagara/MasterMaterials/Functions/Common/PlatformSwitch_3D}
  n73 = SetMaterialAttributes!(
   MaterialAttributes=@n61.MaterialAttributes
   EmissiveColor=StaticSwitchParameter(True=Add(A=@n64, B=@n67), False=@n64) {group=Function_Left_Right_Decal V2-Counter_Reactive, parameterName=FDecalV2_Counter_Reactive, sortPriority=0}
   BaseColor=StaticSwitchParameter(True=Multiply(A=@n68, B=@n69), False=@n69) {group=Function_Left_Right_Decal V2-Counter_Reactive, parameterName=FDecalV2_Counter_Reactive, sortPriority=0}
   Roughness=@n63.Roughness
   Metallic=@n63.Roughness
   Specular=@n63.Roughness
   Normal=StaticSwitchParameter(
    True=StaticSwitchParameter(True=@n72.Result, False=@n71) {defaultValue=true, group=Function_Left_Right_Decal V2, parameterName=FDecalV_Normal__Android Disable}
    False=@n70
   ) {group=Function_Left_Right_Decal V2-Counter, parameterName=FDecalV2_Counter_Normal}
  ) {attributeSetTypes [
    #6
    #1
    #2
    #4
    #3
    #5
   ]}
  n74 = BlendMaterialAttributes(A=@n61.MaterialAttributes, B=@n73, Alpha=NamedRerouteUsage()) {vertexAttributeBlendType=UseA}
  n75 = NamedRerouteUsage()
  n76 = Multiply(
   A=@n75
   B=StaticSwitchParameter(True=@n67, False=@n66) {group=Function_Left_Right_Decal V2-Counter_Reactive, parameterName=FDecalV2_Counter_Reactive, sortPriority=0}
  )
  n77 = ScalarParameter() {defaultValue=1, group=Function_Left_Right_Decal V2-Counter, parameterName="FDecalV2_Counter_Emissive--->0=BaseColor"}
  n78 = StaticSwitchParameter(
   True=@n68
   False=ScalarParameter() {defaultValue=1, group=Function_Left_Right_Decal V2-Counter, parameterName="FDecalV2_Counter_DarkenBase   0=On"}
  ) {group=Function_Left_Right_Decal V2-Counter_Reactive, parameterName=FDecalV2_Counter_Reactive, sortPriority=0}
  n79 = Multiply(A=@n61.BaseColor, B=Max(A=Min(A=@n78, B=@n77), B=OneMinus(None=@n75)))
  n80 = SetMaterialAttributes!(
   MaterialAttributes=@n61.MaterialAttributes
   EmissiveColor=LinearInterpolate(A=@n61.EmissiveColor, B=Add(A=@n61.EmissiveColor, B=@n76), Alpha=@n77)
   BaseColor=LinearInterpolate(A=Add(A=@n79, B=@n76), B=@n79, Alpha=@n77) {constB=0}
  ) {attributeSetTypes [#6, #1]}
  n81 = NamedRerouteUsage()
  n82 = StaticBoolParameter() {group=Function_Left_Right_Decal V2-Counter, parameterName=FDecalV2_Counter}
  n83 = StaticSwitch(True=@n60, False=StaticSwitch(True=StaticSwitch(True=@n74, False=@n80, Value=@n81), False=@n60, Value=@n82), Value=NamedRerouteUsage())
  n84 = FunctionInput(Preview=Constant() {r=1}) {bUsePreviewValueAsDefault=true, inputName="Msk_Ts01<--->Ts02", inputType=FunctionInput_Scalar}
  n85 = LinearInterpolate(
   A=@n84
   B=OneMinus(None=@n84)
   Alpha=ScalarParameter() {group="Function_Left_Right_Decal  V2 msk", parameterName="FDecalV2_Blend TopSide01---> 02 Invert Msk  1=inverted"}
  )
  n86 = StaticBoolParameter() {group="Function_Left_Right_Decal  V2 msk", parameterName="FDecalV2_Blend TopSide01---> 02 --BlendMask"}
  n87 = StaticBoolParameter() {group="TopSide_V02_Layer 01 Blend  02", parameterName="TopS Blend 01<--->02"}
  n88 = StaticSwitch(
   True=StaticSwitch(True=BlendMaterialAttributes(A=@n29, B=@n83, Alpha=@n85) {vertexAttributeBlendType=UseA}, False=@n83, Value=@n86)
   False=@n83
   Value=@n87
  )
  n89 = ScalarParameter() {defaultValue=182, group=Function_Left_Right_Decal V2_Reactive, parameterName="FDecalV2_Opacity_TOD_CC  Index", sortPriority=500}
  n90 = ScalarParameter() {group=00-FrontEnd-Overrides, parameterName=00-FE-TODCC-Override}
  n91 = Add(
   A=@n90
   B=CollectionParameter() {collection=&FortniteMaterialParameters, parameterName=TimeOfDay_0-1}
  )
  n92 = CurveAtlasRowParameter(CurveTime=@n91) {
   atlas=/Game/Weapons/WeaponSkins/Wraps/Textures/CruveAtlas_Wraps_01
   curve=/Game/Weapons/WeaponSkins/Wraps/Textures/0_ColorCurves/Curve_TimeOfDay
   defaultValue=182
   group=Function_Left_Right_Decal V2_Reactive
   parameterName=FDecalV2_Opacity_TOD_CC
   sortPriority=500
  }
  n93 = ChannelMaskParameter(None=MaterialFunctionCall!(Index=@n89, CCinput=@n92, CCinput_Alpha=@n92, UV=@n91) {materialFunction=&/MF_Global_CC_Tex}.4v) {group=Function_Left_Right_Decal V2_Reactive, parameterName=FDecalV2_Opacity_TOD_CC_Channel, sortPriority=500}
  n94 = ScalarParameter() {defaultValue=1, group=Function_Left_Right_Decal V2_Reactive, parameterName=FDecalV2_Opacity_TOD_Morrning, sortPriority=500}
  n95 = Multiply(A=@n90) {constB=-1}
  n96 = ScalarParameter() {defaultValue=1, group=Function_Left_Right_Decal V2_Reactive, parameterName=FDecalV2_Opacity_TOD_Day, sortPriority=500}
  n97 = MaterialFunctionCall(
   Value_Morning=@n94
   Value_Day=Add(A=LinearInterpolate(A=@n90, B=@n95, Alpha=@n96), B=@n96)
   Value_Evening=ScalarParameter() {defaultValue=1, group=Function_Left_Right_Decal V2_Reactive, parameterName=FDecalV2_Opacity_TOD_Evening, sortPriority=500}
   Value_Night=ScalarParameter() {defaultValue=1, group=Function_Left_Right_Decal V2_Reactive, parameterName=FDecalV2_Opacity_TOD_Night, sortPriority=500}
  ) {materialFunction=&Scalar_Time_of_Day}
  n98 = BlendMaterialAttributes(
   A=@n28
   B=@n88
   Alpha=StaticSwitchParameter(True=@n93.R, False=@n97.Result) {group=Function_Left_Right_Decal V2_Reactive, parameterName="FDecalV2_Opacity_TOD_--->CCTOD", sortPriority=500}
  )
  FunctionOutput(None=StaticSwitchParameter(
    True=StaticSwitchParameter(True=@n98, False=@n88) {group=Function_Left_Right_Decal V2_Reactive, parameterName=FDecalV2_Opacity_TOD-------------, sortPriority=500}
    False=@n28
   ) {group=Customization, parameterName=Function_Left_Right_Decal_V2})
  n99 = Multiply(A=NamedRerouteUsage(), B=@n45)
  n100 = NamedRerouteUsage()
  FunctionOutput(None=StaticSwitchParameter(True=LinearInterpolate(B=@n99, Alpha=@n100), False=Constant()) {group=Customization, parameterName=Function_Left_Right_Decal_V2}) {outputName=Decal_Mask_RGBA}
  n101 = Multiply(A=@n45, B=@n46)
  FunctionOutput(None=StaticSwitchParameter(True=LinearInterpolate(B=@n101, Alpha=@n100), False=Constant()) {group=Customization, parameterName=Function_Left_Right_Decal_V2}) {outputName=Decal_Mask_Alpha}
  n102 = ScalarParameter() {defaultValue=1, group=Function_Left_Right_Decal V2_Emissive, parameterName=FDecalV2_Emissive_Gradent_TileRate}
  n103 = MaterialFunctionCall(
   UVs=NamedRerouteUsage()
   "Rotation Angle (0-1)"=ScalarParameter() {group=Function_Left_Right_Decal V2_Emissive, parameterName=FDecalV2_Emissive_Gradent_Rotation}
  ) {materialFunction=&CustomRotator}
  n104 = VectorParameter() {group=Function_Left_Right_Decal V2_Emissive, parameterName=FDecalV2_Emissive_Offset, sortPriority=5}
  n105 = VectorParameter() {group=Function_Left_Right_Decal V2_Emissive, parameterName=FDecalV2_Emissive_Gradent_Scale, defaultValue=<rgba 1 1 1 1>}
  n106 = Multiply(
   A=Add(A=@n103.Rotated Values, B=ComponentMask(None=@n104.RGB) {g=true, r=true})
   B=ComponentMask(None=Multiply(A=@n105.RGB, B=@n105.RGB)) {g=true, r=true}
  )
  n107 = ScalarParameter() {group=Function_Left_Right_Decal V2_Emissive_Reactive, parameterName=FDecalV2_Emissive_Gradent_Time_Start, sortPriority=210}
  n108 = ScalarParameter() {defaultValue=1, group=Function_Left_Right_Decal V2_Emissive_Reactive, parameterName=FDecalV2_Emissive_Gradent_Time_End, sortPriority=210}
  n109 = NamedRerouteUsage()
  n110 = VectorParameter() {group=Function_Left_Right_Decal V2_Emissive, parameterName=FDecalV2_Emissive_Gradent_Offset, defaultValue=<rgba 0 0 0 1>}
  n111 = VectorParameter() {group=Function_Left_Right_Decal V2_Emissive, parameterName=FDecalV2_Emissive_Gradent_Speed, defaultValue=<rgba 0 0 0 1>}
  n112 = Multiply(A=Time(), B=@n111.RGB)
  n113 = StaticBoolParameter() {
   group=Function_Left_Right_Decal V2_Emissive_Reactive
   parameterName=FDecalV2_Emissive_Gradent_Reactive---------------------
   sortPriority=200
  }
  n114 = StaticSwitch(
   True=AppendVector(A=LinearInterpolate(A=@n107, B=@n108, Alpha=@n109), B=Constant())
   False=ComponentMask(None=Add(A=@n110.RGB, B=@n112)) {g=true, r=true}
   Value=@n113
  )
  n115 = Add(A=@n106, B=@n114)
  LinearInterpolate(A=Add(A=Multiply(A=@n106, B=@n102), B=@n114), B=Multiply(A=@n115, B=@n102), Alpha=Ceil(None=@n102))
  n116 = StaticSwitchParameter(
   True=NamedRerouteUsage()
   False=StaticSwitchParameter(True=NamedRerouteUsage(), False=NamedRerouteUsage()) {desc=Look To TopSide Shared, group=Function_Left_Right_Decal V2, parameterName="FDecalV2_Mask__Tex -->TopS_00B_Sharerd_Tex_Msk"}
  ) {desc=Look To TopSide Shared, group=Function_Left_Right_Decal V2, parameterName="FDecalV2_Mask__Tex -->TopS_00A_Sharerd_Tex_Msk"}
  MaterialFunctionCall(UVs=NamedRerouteUsage(), Texture Size=Multiply(A=TextureProperty(Tex=@n116), B=NamedRerouteUsage())) {materialFunction=&ComputeMipLevel}
  n117 = NamedRerouteUsage()
  MaterialFunctionCall(
   "Animation  Phase (0-1)"=@n15
   Number of Rows=@n16
   Number of Columns=@n17
   Texture=@n117
   UVs=@n20
   MipBias/Level=@n25.MipLevel
   "Clamp Anim (see tooltip)"=@n26
  ) {materialFunction=&FlipBook}
  n118 = NamedRerouteUsage()
  n119 = ScalarParameter() {defaultValue=1, group=Function_Left_Right_Decal V2, parameterName=FDecalV2_Miplevel_Adjustment, sortPriority=0}
  MaterialFunctionCall(
   UVs=NamedRerouteUsage()
   Texture Size=Multiply(A=TextureProperty(Tex=@n118), B=NamedRerouteDeclaration(None=@n119) {name=Mip_levels, nodeColor=<rgba 0 1 0.2266 1>})
  ) {materialFunction=&ComputeMipLevel}
  n120 = ChannelMaskParameter(None=NamedRerouteUsage()) {group=Function_Left_Right_Decal V2_Emissive, parameterName=FDecalV2_Emissive_Mask_Channel, sortPriority=5}
  n121 = NamedRerouteUsage()
  Max(A=@n120.R, B=@n121)
  n122 = ChannelMaskParameter(None=NamedRerouteUsage()) {group=Function_Left_Right_Decal V2_Emissive, parameterName=FDecalV2_Emissive_Mask_Channel, sortPriority=5}
  n123 = NamedRerouteUsage()
  Max(A=@n122.R, B=@n123)
  Multiply(A=@n13)
  n124 = VectorParameter() {
   group="Function_Left_Right_Decal  V2 msk"
   parameterName=FDecalV2_TriPlanner_Curvature
   sortPriority=-5
   channelNames {a="", b="", g=Pits, r=peaks}
   defaultValue=<rgba 2 2 0 1>
  }
  n125 = Add(A=Multiply(A=NamedRerouteUsage(), B=@n124.peaks), B=Multiply(A=NamedRerouteUsage(), B=@n124.Pits))
  Multiply(A=@n125) {constB=2}
  Multiply(A=@n100, B=@n99)
  n126 = SetMaterialAttributes!(MaterialAttributes, BaseColor=Constant3Vector() {constant=<rgba 1 0 0 0>}) {attributeSetTypes [#1]}
  NamedRerouteDeclaration(None=GetMaterialAttributes(None=FunctionInput(Preview=@n126) {bUsePreviewValueAsDefault=true, inputName=Material_Root, inputType=FunctionInput_MaterialAttributes, sortPriority=-20}) {attributeGetTypes [#5, #6]}.Normal) {name=Root_Normal, nodeColor=<rgba 0 1 0.3203 1>}
  n127 = FunctionInput(Preview=TextureSample() {samplerType=SAMPLERTYPE_Masks, texture=/Game/Weapons/FORT_Rifles/Textures/T_Scar_CM}.RGBA) {inputName=CM_Msk, inputType=FunctionInput_Vector4, sortPriority=4}
  n128 = MaterialFunctionCall(Float4=@n127) {materialFunction=&BreakOutFloat4Components}
  NamedRerouteDeclaration(None=@n128.G) {name=Msk_Curvature, nodeColor=<rgba 0 1 0.7422 1>}
  n129 = Constant3Vector() {constant=<rgba 0 1 0 0>}
  n130 = GetMaterialAttributes(None=FunctionInput(Preview=SetMaterialAttributes(BaseColor=@n129) {attributeSetTypes [#1]}) {inputName=Material, inputType=FunctionInput_MaterialAttributes, sortPriority=-15}) {attributeGetTypes [
    #1
    #2
    #5
    #4
    #6
   ]}
  NamedRerouteDeclaration(None=@n130.MaterialAttributes) {name=In_Material, nodeColor=<rgba 0.7812 1 0 1>}
  n131 = Constant() {r=1}
  NamedRerouteDeclaration(None=StaticSwitch(True=StaticSwitch(True=@n85, False=@n131, Value=@n86), False=@n131, Value=@n87)) {name="Msk Msk_Ts01<--->Ts02", nodeColor=<rgba 0 0.8594 1 1>}
  n132 = StaticBoolParameter() {group=Function_Left_Right_Decal V2_Emissive, parameterName=FDecalV2_Emissive_Mask_Add_Counter, sortPriority=5}
  NamedRerouteDeclaration(None=@n132) {name=Add_Counter_Stat_B, nodeColor=<rgba 0 0.2969 1 1>}
  n133 = Constant3Vector() {constant=<rgba -1 1 1 1>}
  n134 = ComponentMask(None=NamedRerouteUsage()) {b=true, g=true, r=true}
  n135 = MaterialFunctionCall(Float4=NamedRerouteUsage()) {materialFunction=&BreakOutFloat4Components}
  n136 = LinearInterpolate(A=@n134, B=Multiply(A=@n133, B=@n134), Alpha=@n135.R)
  n137 = ComponentMask(None=NamedRerouteUsage()) {b=true, g=true, r=true}
  n138 = ComponentMask(None=NamedRerouteUsage()) {b=true, g=true, r=true}
  n139 = Multiply(A=@n133, B=@n138)
  n140 = StaticSwitch(
   True=LinearInterpolate(A=@n137, B=Multiply(A=@n133, B=@n137), Alpha=@n135.R)
   False=StaticSwitch(True=LinearInterpolate(A=@n138, B=@n139, Alpha=@n135.R), False=NamedRerouteUsage(), Value=NamedRerouteUsage())
   Value=NamedRerouteUsage()
  )
  n141 = NamedRerouteUsage()
  n142 = LinearInterpolate(
   B=NamedRerouteUsage()
   Alpha=NamedRerouteDeclaration(None=ScalarParameter() {defaultValue=1, group=Function_Left_Right_Decal V2_Uvs, parameterName="FDecalV2_UV_Sine__Account for Rotation = 0"}) {name=Sine Rotation, nodeColor=<rgba 1 0 0.1641 1>}
  )
  NamedRerouteDeclaration(None=Add(A=StaticSwitch(True=@n136, False=@n140, Value=@n141), B=@n142)) {name=Offset_Left_Right, nodeColor=<rgba 0 1 0.9766 1>}
  NamedRerouteDeclaration(None=StaticBoolParameter() {group=Function_Left_Right_Decal V2_Uvs, parameterName=FDecalaV2_UV_SafeZone_Back, sortPriority=0}) {name=SZone_Back_State, nodeColor=<rgba 0 1 0.4141 1>}
  NamedRerouteDeclaration(None=StaticBoolParameter() {group=Function_Left_Right_Decal V2_Uvs, parameterName=FDecalaV2_UV_SafeZone_Middle, sortPriority=-1}) {name=SZone_Mid_State, nodeColor=<rgba 0 1 0.4141 1>}
  NamedRerouteDeclaration(None=FunctionInput(Preview=Constant3Vector() {constant=<rgba 1 0 0 0>}) {bUsePreviewValueAsDefault=true, inputName=ForwardAxis}) {name=Forward_Axies, nodeColor=<rgba 0 0.9297 1 1>}
  n143 = NamedRerouteUsage()
  n144 = MaterialFunctionCall(Color=NamedRerouteUsage()) {materialFunction=&SplitComponents}
  n145 = LinearInterpolate(A=NamedRerouteUsage(), Alpha=NamedRerouteUsage()) {constB=0}
  n146 = NamedRerouteUsage()
  n147 = NamedRerouteUsage()
  n148 = MaterialFunctionCall!(
   For_Safe_Zones=StaticBool() {value=true}
   Normal_Map=NamedRerouteUsage()
   Caculated_Bounds=NamedRerouteUsage()
   Pre_Skinned_Normals=NamedRerouteUsage()
   Is Vehicle=@n144.R
   Is VehicleSide=@n144.R
   Is Mech=@n144.R
   Vehicle_Rotation
   UV_Rotation Bypass=NamedRerouteUsage()
   Mech_Rotation
   Uvs_Offset=NamedRerouteUsage()
   Offset-after rotation=@n145
   Uvs_TileRate=StaticSwitchParameter(True=Multiply(A=@n146, B=@n147), False=@n146) {group=Function_Left_Right_Decal V2_Reactive, parameterName=FDecalV2_Opacity_Reactive---------------------, sortPriority=0}
  ) {materialFunction=&/MF_Projection_Side_Normal}
  n149 = Saturate(None=Add(A=@n143, B=@n148.Safe_Zone_Mask))
  n150 = MaterialFunctionCall(Color=NamedRerouteUsage()) {materialFunction=&SplitComponents}
  n151 = NamedRerouteUsage()
  n152 = NamedRerouteUsage()
  n153 = StaticSwitchParameter(True=Multiply(A=@n152, B=NamedRerouteUsage()), False=@n152) {group=Function_Left_Right_Decal V2_Reactive, parameterName=FDecalV2_Opacity_Reactive---------------------, sortPriority=0}
  n154 = NamedRerouteUsage()
  n155 = MaterialFunctionCall!(
   Normal_Map=NamedRerouteUsage()
   Caculated_Bounds=NamedRerouteUsage()
   Pre_Skinned_Normals=NamedRerouteUsage()
   Is Vehicle=@n150.R
   Is Vehicle side=@n150.R
   Is Mech=@n150.R
   Offset-after rotation=@n145
   UV_Rotation_Right=@n151
   UV_Rotation_Left=Multiply(A=@n151) {constB=-1}
   Uvs_TileRate_Right=@n153
   Uvs_TileRate_Left=@n153
   Uvs_Offset_Right=@n154
   Uvs_Offset_Left=Multiply(A=@n154, B=Constant3Vector() {constant=<rgba -1 1 1 0>})
  ) {materialFunction=&/MF_Projection_Side_Left_Right_Decal_Normal}
  n156 = Saturate(None=Add(A=@n143, B=@n155.Clamp_Mask))
  n157 = NamedRerouteUsage()
  NamedRerouteDeclaration(None=StaticSwitch(True=@n149, False=@n156, Value=@n157)) {name=TileMasking, nodeColor=<rgba 1 0 0.7969 1>}
  NamedRerouteDeclaration(None=StaticBoolParameter() {group=Function_Left_Right_Decal V2_Uvs, parameterName=FDecalaV2_UV_SafeZone_Front, sortPriority=-2}) {name=SZone_Front_State, nodeColor=<rgba 0 1 0.4141 1>}
  n158 = Add(A=NamedRerouteUsage()) {constB=-0.5}
  NamedRerouteDeclaration(None=Saturate(None=Multiply(A=@n158) {constB=-1})) {name=Curve_Pits, nodeColor=<rgba 1 0 0.8906 1>}
  NamedRerouteDeclaration(None=Saturate(None=@n158)) {name=Curve_Peaks, nodeColor=<rgba 1 0 0.8906 1>}
  n159 = TextureObjectParameter() {
   group=TopSide_V02
   parameterName=TopS_00B_Sharerd_Tex_Msk
   samplerType=SAMPLERTYPE_Masks
   sortPriority=249
   texture=&T_BlackMask
  }
  NamedRerouteDeclaration(None=@n159) {name=Tex_00B_Shared_Tex, nodeColor=<rgba 1 0.8906 0 1>}
  n160 = ComponentMask(None=NamedRerouteUsage()) {a=true}
  n161 = NamedRerouteUsage()
  n162 = Multiply(A=ComponentMask(None=NamedRerouteUsage()) {a=true}, B=@n161)
  n163 = NamedRerouteUsage()
  n164 = StaticSwitch(True=Multiply(A=ComponentMask(None=@n163) {a=true}, B=@n161), False=NamedRerouteUsage(), Value=NamedRerouteUsage())
  n165 = NamedRerouteUsage()
  NamedRerouteDeclaration(None=StaticSwitch(True=Multiply(A=@n160, B=@n161), False=StaticSwitch(True=@n162, False=@n164, Value=@n165), Value=NamedRerouteUsage())) {name=Tiles_Rate_States, nodeColor=<rgba 0.5469 1 0 1>}
  n166 = VectorParameter() {group=Function_Left_Right_Decal V2_Uvs, parameterName=FDecalV2_UV_Scale_Weapon, defaultValue=<rgba 1 1 1 1>}
  n167 = VectorParameter() {group=Function_Left_Right_Decal V2_Uvs, parameterName=FDecalV2_UV_Scale_Vehicle, defaultValue=<rgba 1 1 1 1>}
  n168 = Multiply(A=@n167.RGB, B=@n167.RGB)
  n169 = VectorParameter() {group=Function_Left_Right_Decal V2_Uvs, parameterName=FDecalV2_UV_Scale_VehicleSide, defaultValue=<rgba 1 1 1 1>}
  n170 = Multiply(A=@n169.RGB, B=@n169.RGB)
  n171 = MaterialFunctionCall(Float4=NamedRerouteUsage()) {materialFunction=&BreakOutFloat4Components}
  n172 = VectorParameter() {group=Function_Left_Right_Decal V2_Uvs, parameterName=FDecalV2_UV_Scale_Mech, defaultValue=<rgba 1 1 1 1>}
  n173 = VectorParameter() {group=Function_Left_Right_Decal V2_Uvs, parameterName=FDecalV2_UV_Scale_Character, defaultValue=<rgba 1 1 1 1>}
  n174 = LinearInterpolate(
   A=LinearInterpolate(A=Multiply(A=@n166.RGB, B=@n166.RGB), B=LinearInterpolate(A=@n168, B=@n170, Alpha=@n171.G), Alpha=@n171.R)
   B=LinearInterpolate(A=Multiply(A=@n172.RGB, B=@n172.RGB), B=Multiply(A=@n173.RGB, B=@n173.RGB), Alpha=@n171.A)
   Alpha=@n171.B
  )
  n175 = VectorParameter() {group=Function_Left_Right_Decal V2_Uvs, parameterName=FDecalV2_UV_Scale_Backpack, defaultValue=<rgba 1 1 1 1>}
  n176 = Multiply(A=@n175.RGB, B=@n175.RGB)
  n177 = MaterialFunctionCall(Float2=NamedRerouteUsage()) {materialFunction=&BreakOutFloat2Components}
  n178 = VectorParameter() {group=Function_Left_Right_Decal V2_Uvs, parameterName=FDecalV2_UV_Scale_Pickaxe, defaultValue=<rgba 1 1 1 1>}
  NamedRerouteDeclaration(None=LinearInterpolate(A=LinearInterpolate(A=@n174, B=@n176, Alpha=@n177.R), B=Multiply(A=@n178.RGB, B=@n178.RGB), Alpha=@n177.G)) {name=TileRate, nodeColor=<rgba 0.0781 0 1 1>}
  n179 = ScalarParameter() {group=Function_Left_Right_Decal V2_Uvs, parameterName=FDecalV2_UV_Rotation_Vehicle}
  n180 = ScalarParameter() {group=Function_Left_Right_Decal V2_Uvs, parameterName=FDecalV2_UV_Rotation_VehicleSide}
  n181 = MaterialFunctionCall(Float4=NamedRerouteUsage()) {materialFunction=&BreakOutFloat4Components}
  n182 = LinearInterpolate(
   A=LinearInterpolate(
    A=ScalarParameter() {group=Function_Left_Right_Decal V2_Uvs, parameterName=FDecalV2_UV_Rotation_Weapon}
    B=LinearInterpolate(A=@n179, B=@n180, Alpha=@n181.G)
    Alpha=@n181.R
   )
   B=LinearInterpolate(
    A=ScalarParameter() {defaultValue=0.25, group=Function_Left_Right_Decal V2_Uvs, parameterName=FDecalV2_UV_Rotation_Mech}
    B=ScalarParameter() {defaultValue=0.25, group=Function_Left_Right_Decal V2_Uvs, parameterName=FDecalV2_UV_Rotation_Character}
    Alpha=@n181.A
   )
   Alpha=@n181.B
  )
  n183 = ScalarParameter() {group=Function_Left_Right_Decal V2_Uvs, parameterName=FDecalV2_UV_Rotation_Backpack}
  n184 = MaterialFunctionCall(Float2=NamedRerouteUsage()) {materialFunction=&BreakOutFloat2Components}
  NamedRerouteDeclaration(None=LinearInterpolate(
    A=LinearInterpolate(A=@n182, B=@n183, Alpha=@n184.R)
    B=ScalarParameter() {group=Function_Left_Right_Decal V2_Uvs, parameterName=FDecalV2_UV_Rotation_Pickaxe}
    Alpha=@n184.G
   )) {name=Rotation, nodeColor=<rgba 0 0.5547 1 1>}
  n185 = TextureObjectParameter() {group=Function_Left_Right_Decal V2, parameterName=FDecalV2_BaseColor_Tex, texture=&T_BlacksRGB}
  NamedRerouteDeclaration(None=@n185) {name=Tex_Sample_BC, nodeColor=<rgba 0.1016 1 0 1>}
  NamedRerouteDeclaration(None=Abs(None=ScalarParameter() {group=Function_Left_Right_Decal V2-Counter, parameterName=FDecalV2_Counter__Number})) {name=Count_Number, nodeColor=<rgba 1 0 0.0469 1>}
  n186 = NamedRerouteUsage()
  n187 = ComponentMask(None=VectorParameter() {group=Function_Left_Right_Decal V2-Counter, parameterName=FDecalV2_Counter_UV_Offset}.RGB) {g=true, r=true}
  n188 = VectorParameter() {
   group=Function_Left_Right_Decal V2-Counter
   parameterName=FDecalV2_Counter_UV_Scale
   channelNames {a=Uniform, b="", g="", r=""}
   defaultValue=<rgba 1 1 0 0.5>
  }
  n189 = MaterialFunctionCall(
   UVs=MaterialFunctionCall(
    UVs=Add(A=@n186, B=@n187)
    "Rotation Angle (0-1)"=ScalarParameter() {group=Function_Left_Right_Decal V2-Counter, parameterName=FDecalV2_Counter_UV_Rotation}
   ) {materialFunction=&CustomRotator}.Rotated Values
   Texture Scale=ComponentMask(None=Multiply(A=@n188, B=@n188)) {g=true, r=true}
  ) {materialFunction=&ScaleUVsByCenter}
  n190 = Multiply(A=@n189.UVs, B=Constant2Vector() {g=0.1, r=1}) {constA=2, constB=0.1}
  n191 = ComponentMask(None=@n190) {r=true}
  n192 = ScalarParameter() {defaultValue=3, desc="How many Coulums ", group=Function_Left_Right_Decal V2-Counter, parameterName=FDecalV2_Counter__Number_Max}
  NamedRerouteDeclaration(None=Multiply(A=OneMinus(None=@n191), B=@n192) {constB=10}) {name=Count_Max_Number, nodeColor=<rgba 0 0.3672 1 1>}
  n193 = ComponentMask(None=NamedRerouteUsage()) {b=true, g=true, r=true}
  n194 = NamedRerouteUsage()
  n195 = StaticSwitch(
   True=ComponentMask(None=NamedRerouteUsage()) {b=true, g=true, r=true}
   False=StaticSwitch(True=ComponentMask(None=@n194) {b=true, g=true, r=true}, False=NamedRerouteUsage(), Value=NamedRerouteUsage())
   Value=NamedRerouteUsage()
  )
  n196 = NamedRerouteUsage()
  NamedRerouteDeclaration(None=Add(A=StaticSwitch(True=@n193, False=@n195, Value=@n196), B=@n142)) {name=Offset_Side, nodeColor=<rgba 1 0 0.8672 1>}
  n197 = TextureObjectParameter() {
   group=Function_Left_Right_Decal V2
   parameterName=FDecalV2_Normal_Tex
   samplerType=SAMPLERTYPE_Normal
   texture=&FlatNormalBlue
  }
  NamedRerouteDeclaration(None=@n197) {name=Tex_Normal, nodeColor=<rgba 0 1 0.8828 1>}
  n198 = VectorParameter() {group=Function_Left_Right_Decal V2_Uvs, parameterName=FDecalV2_UV_Offset_Weapon, defaultValue=<rgba 0 0 0 1>}
  n199 = VectorParameter() {group=Function_Left_Right_Decal V2_Uvs, parameterName=FDecalV2_UV_Offset_Vehicle, defaultValue=<rgba 0 0 0 1>}
  n200 = VectorParameter() {group=Function_Left_Right_Decal V2_Uvs, parameterName=FDecalV2_UV_Offset_VehicleSide, defaultValue=<rgba 0 0 0 1>}
  n201 = MaterialFunctionCall(Float4=NamedRerouteUsage()) {materialFunction=&BreakOutFloat4Components}
  n202 = VectorParameter() {group=Function_Left_Right_Decal V2_Uvs, parameterName=FDecalV2_UV_Offset_Mech, defaultValue=<rgba 0 0 0 1>}
  n203 = VectorParameter() {group=Function_Left_Right_Decal V2_Uvs, parameterName=FDecalV2_UV_Offset_Character, defaultValue=<rgba 0 0 0 1>}
  n204 = LinearInterpolate(
   A=LinearInterpolate(A=@n198.RGB, B=LinearInterpolate(A=@n199.RGB, B=@n200.RGB, Alpha=@n201.G), Alpha=@n201.R)
   B=LinearInterpolate(A=@n202.RGB, B=@n203.RGB, Alpha=@n201.A)
   Alpha=@n201.B
  )
  n205 = VectorParameter() {group=Function_Left_Right_Decal V2_Uvs, parameterName=FDecalV2_UV_Offset_Backpack, defaultValue=<rgba 0 0 0 1>}
  n206 = MaterialFunctionCall(Float2=NamedRerouteUsage()) {materialFunction=&BreakOutFloat2Components}
  n207 = VectorParameter() {group=Function_Left_Right_Decal V2_Uvs, parameterName=FDecalV2_UV_Offset_Pickaxe, defaultValue=<rgba 0 0 0 1>}
  NamedRerouteDeclaration(None=LinearInterpolate(A=LinearInterpolate(A=@n204, B=@n205.RGB, Alpha=@n206.R), B=@n207.RGB, Alpha=@n206.G)) {name=Offset, nodeColor=<rgba 0 0.5547 1 1>}
  n208 = Multiply(
   A=Time()
   B=ScalarParameter() {group=Function_Left_Right_Decal V2_Uvs, parameterName="FDecalV2_UV_Rotation_Speed  --------------", sortPriority=1}
  )
  n209 = ScalarParameter() {defaultValue=0.1, group=Function_Left_Right_Decal V2_Uvs, parameterName=FDecalV2_UV_Rotation_Speed Bounce Strength, sortPriority=1}
  n210 = Multiply(A=Sine(None=@n208), B=@n209)
  n211 = Ceil(None=ScalarParameter() {group=Function_Left_Right_Decal V2_Uvs, parameterName="FDecalV2_UV_Rotation_Speed  1=Bounce", sliderMax=1, sortPriority=1})
  NamedRerouteDeclaration(None=Add(A=NamedRerouteUsage(), B=LinearInterpolate(A=@n208, B=@n210, Alpha=@n211))) {name="Rotation\\Speed", nodeColor=<rgba 0 1 0.8125 1>}
  n212 = TextureObjectParameter() {
   group=TopSide_V02
   parameterName=TopS_00A_Sharerd_Tex_Msk
   samplerType=SAMPLERTYPE_Masks
   sortPriority=249
   texture=&T_BlackMask
  }
  NamedRerouteDeclaration(None=@n212) {name=Tex_00A_Shared_Tex, nodeColor=<rgba 1 0.8906 0 1>}
  NamedRerouteDeclaration(None=@n127) {name=CM_Mask, nodeColor=<rgba 1 0 0.3047 1>}
  NamedRerouteDeclaration(None=@n128.R) {name=Msk_Prime, nodeColor=<rgba 0 1 0.7422 1>}
  NamedRerouteDeclaration(None=StaticBoolParameter() {group=Function_Left_Right_Decal V2, parameterName="FDecalaV2_Projection-LR--->Side", sortPriority=0}) {name="Projection-LR--->Side State", nodeColor=<rgba 0 1 0.4141 1>}
  NamedRerouteDeclaration(None=@n128.B) {name=Msk_Accent, nodeColor=<rgba 0 1 0.7656 1>}
  n213 = StaticSwitch(True=@n148.UVs, False=@n155.UVs, Value=NamedRerouteUsage())
  n214 = NamedRerouteUsage()
  NamedRerouteDeclaration(None=StaticSwitchParameter(True=Add(A=@n213, B=@n214), False=@n213) {group=Function_Left_Right_Decal V2_Uvs, parameterName=FDecalV2_UV_Warp---------}) {name=UVs, nodeColor=<rgba 0.125 0 1 1>}
  n215 = StaticSwitchParameter(True=NamedRerouteUsage(), False=NamedRerouteUsage()) {desc="By passes merged wrap normals. ", group=Function_Left_Right_Decal V2, parameterName=FDecalV2_Normal__Use_Root_Material_Normal}
  n216 = NamedRerouteUsage()
  n217 = StaticSwitch(
   True=NamedRerouteUsage()
   False=StaticSwitch(True=Multiply(A=@n149, B=@n148.Raw_Normal), False=Multiply(A=@n156, B=@n155.Raw_Normal), Value=@n157)
   Value=NamedRerouteUsage()
  )
  n218 = NamedRerouteUsage()
  n219 = StaticSwitchParameter(True=NamedRerouteUsage(), False=NamedRerouteUsage()) {desc="By passes merged wrap normals. ", group=Function_Left_Right_Decal V2, parameterName=FDecalV2_Normal__Use_Root_Material_Normal}
  n220 = LinearInterpolate(
   A=@n219
   B=Add(A=@n219, B=StaticSwitch(True=@n216, False=@n217, Value=@n218))
   Alpha=ScalarParameter() {defaultValue=1, group=Function_Left_Right_Decal V2, parameterName=FDecalV2_Normal_Intensty}
  )
  n221 = MaterialFunctionCall("Xb1,PS4"=@n220, "PC-High&&Epic,XSX,PS5"=@n220, "PC-Perf,Android"=@n219, Switch=@n220, "PC-Low(NotPerf)&&Medium"=@n220) {materialFunction=&PlatformSwitch_3D}
  NamedRerouteDeclaration(None=StaticSwitchParameter(
    True=@n215
    False=StaticSwitchParameter(True=@n221.Result, False=@n220) {defaultValue=true, group=Function_Left_Right_Decal V2, parameterName=FDecalV_Normal__Android Disable}
   ) {defaultValue=true, group=Function_Left_Right_Decal V2, parameterName=FDecalV2_Normal__Simple}) {name=Normals, nodeColor=<rgba 0 1 0.9297 1>}
  n222 = Constant4Vector() {constant=<rgba 0 0 0 0.5>}
  NamedRerouteDeclaration(None=FunctionInput(Preview=@n222.RGBA) {bUsePreviewValueAsDefault=true, inputName=Object-SafeZone-Back, inputType=FunctionInput_Vector4, sortPriority=100}) {name=SafeZone_Back, nodeColor=<rgba 0.1016 1 0 1>}
  NamedRerouteDeclaration(None=FunctionInput(Preview=@n222.RGBA) {bUsePreviewValueAsDefault=true, inputName=Object-SafeZone-Middle, inputType=FunctionInput_Vector4, sortPriority=101}) {name=SafeZone_Middle, nodeColor=<rgba 0.1016 1 0 1>}
  NamedRerouteDeclaration(None=FunctionInput(Preview=@n222.RGBA) {bUsePreviewValueAsDefault=true, inputName=Object-SafeZone-Front, inputType=FunctionInput_Vector4, sortPriority=102}) {name=SafeZone_Front, nodeColor=<rgba 0.1016 1 0 1>}
  n223 = NamedRerouteUsage()
  n224 = Divide(A=ComponentMask(None=Add(A=@n223) {constB=-0.5}) {g=true}) {constB=-1}
  NamedRerouteDeclaration(None=Saturate(None=Ceil(None=@n224))) {name=Character mask, nodeColor=<rgba 0.9453 1 0 1>}
  n225 = NamedRerouteUsage()
  n226 = ComponentMask(None=NamedRerouteUsage()) {r=true}
  n227 = LinearInterpolate(B=Saturate(None=Add(A=@n225, B=@n226)), Alpha=ComponentMask(None=NamedRerouteUsage()) {a=true}) {constA=1}
  n228 = ScalarParameter() {group="Function_Left_Right_Decal  V2 msk", parameterName="FDecalV2_Remove_Decal_Front_Character   0=On", sliderMax=1, sortPriority=-10}
  n229 = Ceil(None=@n228)
  NamedRerouteDeclaration(None=Saturate(None=Add(A=@n227, B=@n229))) {name=character_Masking_Fnt/bck, nodeColor=<rgba 0 1 0.6016 1>}
  n230 = ScalarParameter() {group="Function_Left_Right_Decal  V2 msk", parameterName="FDecalV2_Clamp_Mask disable 1=Active", sliderMax=1, sortPriority=-4}
  NamedRerouteDeclaration(None=Saturate(None=Round(None=@n230))) {name=ClampMask State, nodeColor=<rgba 0 0.0625 1 1>}
  n231 = MaterialFunctionCall(
   "Animation  Phase (0-1)"=@n15
   Number of Rows=@n16
   Number of Columns=@n17
   Texture=@n22
   UVs=@n20
   MipBias/Level=@n25.MipLevel
   "Clamp Anim (see tooltip)"=@n26
  ) {materialFunction=&FlipBook}
  n232 = DDX(Value=@n21)
  n233 = DDY(Value=@n21)
  n234 = TextureSample!(UVs=@n231.UVs, Tex=@n22, "DDX(UVs)"=@n232, "DDY(UVs)"=@n233, Apply View MipBias) {mipValueMode=TMVM_Derivative}
  NamedRerouteDeclaration(None=@n234.RGB) {name=Flipbook_RGB, nodeColor=<rgba 0 0.0156 1 1>}
  NamedRerouteDeclaration(None=FunctionInput(Preview=Constant4Vector().RGBA) {bUsePreviewValueAsDefault=true, inputName=IsState, inputType=FunctionInput_Vector4, sortPriority=4}) {name=IsState, nodeColor=<rgba 1 0.1131 0.8487 1>}
  NamedRerouteDeclaration(None=@n234.A) {name=Flipbook_Alpha, nodeColor=<rgba 0 0.0156 1 1>}
  NamedRerouteDeclaration(None=StaticBoolParameter() {group=Function_Left_Right_Decal V2, parameterName="FDecalV2_Decal ---->FlipBook", sortPriority=0}) {name=FlipBook_State, nodeColor=<rgba 0.8516 0 1 1>}
  NamedRerouteDeclaration(None=@n227) {name=character_Masking_Fnt/bck_1, nodeColor=<rgba 0 1 0.6016 1>}
  n235 = ScalarParameter() {defaultValue=1, group="Function_Left_Right_Decal  V2 msk", parameterName="FDecalV2_TriPlanner Mask 0=On 1=Off", sliderMax=1, sortPriority=-5}
  n236 = MaterialFunctionCall(Color=Abs(None=NamedRerouteUsage())) {materialFunction=&SplitComponents}
  n237 = MaterialFunctionCall(Color=NamedRerouteUsage()) {materialFunction=&SplitComponents}
  n238 = LinearInterpolate(A=@n236.R, B=LinearInterpolate(A=@n236.B, B=@n236.B, Alpha=@n237.G), Alpha=@n237.R)
  n239 = ScalarParameter() {defaultValue=1, group="Function_Left_Right_Decal  V2 msk", parameterName=FDecalV2_TriPlanner nMask_Contrast, sortPriority=-5}
  n240 = MaterialFunctionCall(In=Add(A=@n125, B=LinearInterpolate(A=@n238, B=@n236.G, Alpha=@n237.B)), Contrast=@n239) {materialFunction=&CheapContrast}
  NamedRerouteDeclaration(None=Saturate(None=Add(A=@n235, B=@n240.Result))) {name=Tri_planner_Msk, nodeColor=<rgba 0.3594 0 1 1>}
  n241 = Time()
  n242 = VectorParameter() {group=Function_Left_Right_Decal V2_Uvs, parameterName=FDecalV2_UV_Sine__SpeedXY, defaultValue=<rgba 0.1 0.1 0 1>}
  n243 = VectorParameter() {group=Function_Left_Right_Decal V2_Uvs, parameterName=FDecalV2_UV_Sine__OffsetXY, defaultValue=<rgba 0 0 0 1>}
  n244 = Sine(None=Add(A=Multiply(A=@n241, B=@n242.RGB), B=@n243.RGB))
  n245 = VectorParameter() {group=Function_Left_Right_Decal V2_Uvs, parameterName=FDecalV2_UV_Sine_RangeXY_Weapon, defaultValue=<rgba 0 0 0 1>}
  n246 = VectorParameter() {group=Function_Left_Right_Decal V2_Uvs, parameterName=FDecalV2_UV_Sine_RangeXY_Vehicle, defaultValue=<rgba 0 0 0 1>}
  n247 = VectorParameter() {group=Function_Left_Right_Decal V2_Uvs, parameterName=FDecalV2_UV_Sine_RangeXY_VehicleSide, defaultValue=<rgba 0 0 0 1>}
  n248 = MaterialFunctionCall(Float4=NamedRerouteUsage()) {materialFunction=&BreakOutFloat4Components}
  n249 = LinearInterpolate(A=@n245.RGB, B=LinearInterpolate(A=@n246.RGB, B=@n247.RGB, Alpha=@n248.G), Alpha=@n248.R)
  n250 = VectorParameter() {group=Function_Left_Right_Decal V2_Uvs, parameterName=FDecalV2_UV_Sine_RangeXY_Mech, defaultValue=<rgba 0 0 0 1>}
  n251 = VectorParameter() {group=Function_Left_Right_Decal V2_Uvs, parameterName=FDecalV2_UV_Sine_RangeXY_Character, defaultValue=<rgba 0 0 0 1>}
  n252 = LinearInterpolate(A=@n250.RGB, B=@n251.RGB, Alpha=@n248.A)
  n253 = VectorParameter() {group=Function_Left_Right_Decal V2_Uvs, parameterName=FDecalV2_UV_Sine_RangeXY_Backpack, defaultValue=<rgba 0 0 0 1>}
  n254 = MaterialFunctionCall(Float2=NamedRerouteUsage()) {materialFunction=&BreakOutFloat2Components}
  n255 = VectorParameter() {group=Function_Left_Right_Decal V2_Uvs, parameterName=FDecalV2_UV_Sine_RangeXY_Pickaxe, defaultValue=<rgba 0 0 0 1>}
  n256 = LinearInterpolate(A=LinearInterpolate(A=LinearInterpolate(A=@n249, B=@n252, Alpha=@n248.B), B=@n253.RGB, Alpha=@n254.R), B=@n255.RGB, Alpha=@n254.G)
  MaterialExpressionAppendVector_0 = AppendVector(A=Multiply(A=ComponentMask(None=@n244) {g=true, r=true}, B=ComponentMask(None=@n256) {g=true, r=true}), B=Constant())
  n257 = Convert(None=@MaterialExpressionAppendVector_0) {
   nodeName=Break Float3
   convertInputs [{
     type=Vector3
     expressionInput {
      expression=&self:MaterialExpressionAppendVector_0
      inputName=None
      mask=0
      maskA=0
      maskB=0
      maskG=0
      maskR=0
      outputIndex=0
     }
    }]
   convertOutputs [{type=Scalar}, {type=Scalar}, {type=Scalar}]
  }
  NamedRerouteDeclaration(None=LinearInterpolate(
    A=MaterialFunctionCall(R=@n257, G=@n257, B=@n257) {materialFunction=&AppendMany}.RGB
    B=@MaterialExpressionAppendVector_0
    Alpha=NamedRerouteUsage()
   )) {name=UVs_Sine_Range, nodeColor=<rgba 0 0.625 1 1>}
  NamedRerouteDeclaration(None=@n234.RGBA) {name=Flipbook RGBA, nodeColor=<rgba 0 0.0156 1 1>}
  NamedRerouteDeclaration(None=@n240.Result) {name=Tri_planner_Msk_1, nodeColor=<rgba 0.3594 0 1 1>}
  NamedRerouteDeclaration(None=StaticSwitch(True=@n148.UVs No Offsets/Tiling, False=@n155.UVs No Offsets/Tiling, Value=NamedRerouteUsage())) {name=UVs_Raw, nodeColor=<rgba 0.125 0 1 1>}
  n258 = NamedRerouteUsage()
  n259 = ScalarParameter() {group=Function_Left_Right_Decal V2_Emissive, parameterName="FDecalV2_Emissive_TriPlanner_Mask  1=ON", sliderMax=1, sortPriority=1}
  n260 = LinearInterpolate(A=@n258, B=NamedRerouteUsage(), Alpha=Ceil(None=@n259))
  NamedRerouteDeclaration(None=@n260) {name=trimasking, nodeColor=<rgba 0.2891 0 1 1>}
  n261 = NamedRerouteUsage()
  n262 = Max(A=NamedRerouteUsage(), B=@n261)
  n263 = Min(A=@n262, B=@n260)
  n264 = NamedRerouteUsage()
  n265 = StaticSwitch(True=@n263, False=Min(A=@n260, B=NamedRerouteUsage()), Value=@n264)
  n266 = NamedRerouteUsage()
  n267 = NamedRerouteUsage()
  NamedRerouteDeclaration(None=StaticSwitch(True=@n263, False=StaticSwitch(True=@n263, False=@n265, Value=@n266), Value=@n267)) {name=Masking_Cust/Tri_Emissive, nodeColor=<rgba 0.2188 0 1 1>}
  NamedRerouteDeclaration(None=FunctionInput(Preview=Constant4Vector().RGBA) {bUsePreviewValueAsDefault=true, inputName=IsState_BV4, inputType=FunctionInput_Vector4, sortPriority=4}) {name=IsState_BV4, nodeColor=<rgba 1 0.1131 0.8487 1>}
  NamedRerouteDeclaration(None=TextureSample(UVs=@n231.UVs, Tex=@n18, "DDX(UVs)"=@n232, "DDY(UVs)"=@n233) {mipValueMode=TMVM_Derivative}.RGBA) {name=Flipbook Msk, nodeColor=<rgba 0 0.0156 1 1>}
  n268 = Min(A=@n262, B=@n258)
  n269 = StaticSwitch(True=@n268, False=Min(A=NamedRerouteUsage(), B=@n258), Value=@n264)
  NamedRerouteDeclaration(None=StaticSwitch(True=@n268, False=StaticSwitch(True=@n268, False=@n269, Value=@n266), Value=@n267)) {name=Masking_Cust/Tri, nodeColor=<rgba 0.2188 0 1 1>}
  n270 = NamedRerouteUsage()
  n271 = ScalarParameter() {defaultValue=1, group="Function_Left_Right_Decal  V2 msk", parameterName="FDecalV2_Primary Mask  0=Active", sliderMax=1, sortPriority=-10}
  n272 = Min(A=@n271, B=@n261)
  NamedRerouteDeclaration(None=Max(
    A=Max(A=@n270, B=@n272)
    B=ScalarParameter() {group="Function_Left_Right_Decal  V2 msk", parameterName="FDecalV2_Disable Mask  1=Active", sliderMax=1, sortPriority=-10}
   )) {name=Cust_Mask, nodeColor=<rgba 0.3125 1 0 1>}
  NamedRerouteDeclaration(None=MaterialFunctionCall!(
    Normal_Map
    Normals_Source=TextureSample(UVs=@n231.UVs, Tex=@n117, "DDX(UVs)"=@n232, "DDY(UVs)"=@n233) {mipValueMode=TMVM_Derivative}.RGB
    Normals_Source_Use=StaticBool() {value=true}
    UVs
    Rotation_Comp
    UV_Rotation=NamedRerouteUsage()
    Pre_Skinned_Normals=NamedRerouteUsage()
    is_State=NamedRerouteUsage()
    "Left\\Right"=NamedRerouteUsage()
    No_Rotation
   ) {materialFunction=&/MF_UVs_Side_Fix_Normals}.Raw_Normal) {name=Flipbook Normals, nodeColor=<rgba 0 0.0156 1 1>}
  n273 = VectorParameter() {group=Function_Left_Right_Decal V2, parameterName="FDecalV2_Base_RGBA_Color=Alpha", channelNames {a="1=Use", b="", g="", r=""}}
  n274 = NamedRerouteUsage()
  n275 = LinearInterpolate(A=Saturate(None=@n274), B=@n274, Alpha=NamedRerouteUsage())
  n276 = NamedRerouteUsage()
  n277 = StaticSwitchParameter(True=NamedRerouteUsage(), False=NamedRerouteUsage()) {group=Function_Left_Right_Decal V2, parameterName="FDecalV2_Base_RGBA_Tex-->TopS_00B_Sharerd_Tex_Msk"}
  n278 = NamedRerouteUsage()
  n279 = DDX(Value=@n278)
  n280 = DDY(Value=@n278)
  n281 = TextureSample(
   UVs=@n275
   Tex=NamedRerouteDeclaration(None=StaticSwitchParameter(True=@n276, False=@n277) {group=Function_Left_Right_Decal V2, parameterName="FDecalV2_Base_RGBA_Tex-->TopS_00A_Sharerd_Tex_Msk"}) {name=Sha, nodeColor=<rgba 1 0.8438 0 1>}
   "DDX(UVs)"=@n279
   "DDY(UVs)"=@n280
  ) {mipValueMode=TMVM_Derivative}
  n282 = LinearInterpolate(B=@n273.RGB, Alpha=Min(A=@n273."1=Use", B=@n281.A)) {constA=1}
  n283 = VectorParameter() {
   group=Function_Left_Right_Decal V2
   parameterName="FDecalV2_Base_RGBA_Color=Blue"
   channelNames {a="1=Use", b="", g="", r=""}
   defaultValue=<rgba 0 0 1 0>
  }
  n284 = Min(A=@n283."1=Use", B=@n281.B)
  n285 = VectorParameter() {
   group=Function_Left_Right_Decal V2
   parameterName="FDecalV2_Base_RGBA_Color=Green"
   channelNames {a="1=Use", b="", g="", r=""}
   defaultValue=<rgba 0 1 0 0>
  }
  n286 = VectorParameter() {
   group=Function_Left_Right_Decal V2
   parameterName="FDecalV2_Base_RGBA_Color=Red"
   channelNames {a="1=Use", b="", g="", r=""}
   defaultValue=<rgba 1 0 0 0>
  }
  n287 = LinearInterpolate(
   A=LinearInterpolate(A=LinearInterpolate(A=@n282, B=@n283.RGB, Alpha=@n284), B=@n285.RGB, Alpha=Min(A=@n281.G, B=@n285."1=Use"))
   B=@n286.RGB
   Alpha=Min(A=@n281.R, B=@n286."1=Use")
  )
  n288 = NamedRerouteUsage()
  n289 = TextureSample(UVs=@n275, Tex=@n118, "DDX(UVs)"=@n279, "DDY(UVs)"=@n280) {mipValueMode=TMVM_Derivative}
  n290 = NamedRerouteUsage()
  n291 = NamedRerouteUsage()
  n292 = StaticSwitch(
   True=NamedRerouteUsage()
   False=StaticSwitch(True=ComponentMask(None=@n288) {b=true, g=true, r=true}, False=@n289.RGB, Value=@n290)
   Value=@n291
  )
  n293 = NamedRerouteUsage()
  n294 = NamedRerouteUsage()
  n295 = StaticSwitch(True=LinearInterpolate(A=NamedRerouteUsage(), B=@n292, Alpha=@n293), False=@n292, Value=@n294)
  NamedRerouteDeclaration(None=StaticSwitchParameter(True=Multiply(A=@n287, B=@n295), False=@n295) {group=Function_Left_Right_Decal V2, parameterName=FDecalV2_Base_RGBA-Color------}) {name=BaseColor, nodeColor=<rgba 0 0.0625 1 1>}
  n296 = TextureSample(UVs=@n275, Tex=NamedRerouteUsage(), "DDX(UVs)"=@n279, "DDY(UVs)"=@n280) {mipValueMode=TMVM_Derivative}
  n297 = NamedRerouteUsage()
  n298 = StaticSwitch(True=NamedRerouteUsage(), False=@n296.A, Value=@n297)
  n299 = StaticSwitch(True=NamedRerouteUsage(), False=StaticSwitch(True=ComponentMask(None=@n288) {a=true}, False=@n289.A, Value=@n290), Value=@n291)
  n300 = NamedRerouteUsage()
  n301 = StaticSwitch(True=NamedRerouteUsage(), False=@n293, Value=@n300)
  n302 = StaticBool() {value=true}
  NamedRerouteDeclaration(None=StaticSwitch(True=LinearInterpolate(A=@n298, B=@n299, Alpha=@n301), False=@n299, Value=StaticSwitch(True=@n302, False=@n294, Value=@n300))) {name=BaseColor_Alpha, nodeColor=<rgba 0 0.0625 1 1>}
  n303 = NamedRerouteUsage()
  n304 = NamedRerouteUsage()
  n305 = VectorParameter() {group=Function_Left_Right_Decal V2_Uvs, parameterName=FDecalV2_Mask_TileRate_UV, defaultValue=<rgba 1 1 0 1>}
  n306 = Multiply(A=@n304, B=ComponentMask(None=@n305.RGB) {g=true, r=true})
  n307 = VectorParameter() {group=Function_Left_Right_Decal V2_Uvs, parameterName=FDecalV2_Mask_Speed_UV, defaultValue=<rgba 0 0 0 1>}
  n308 = Multiply(A=Time(), B=ComponentMask(None=@n307.RGB) {g=true, r=true})
  n309 = TextureSample(
   UVs=LinearInterpolate(
    A=Add(A=@n306, B=@n308)
    B=Saturate(None=@n304)
    Alpha=ScalarParameter() {group=Function_Left_Right_Decal V2, parameterName="FDecalV2_Mask_Clamp_UV  1=Active"}
   )
   Tex=@n116
   "DDX(UVs)"=@n279
   "DDY(UVs)"=@n280
  ) {mipValueMode=TMVM_Derivative}
  n310 = NamedRerouteUsage()
  NamedRerouteDeclaration(None=StaticSwitch(True=NamedRerouteUsage(), False=StaticSwitch(True=@n303, False=@n309.RGBA, Value=@n310), Value=NamedRerouteUsage())) {name=Mask_RGBA, nodeColor=<rgba 1 0 0.7031 1>}
  n311 = ScalarParameter() {defaultValue=-9999999, group=Function_Left_Right_Decal V2_Reactive, parameterName=FDecalaV2_Opacity_Time_BP_Z}
  n312 = ScalarParameter() {group=Function_Left_Right_Decal V2_Reactive, parameterName=FDecalaV2_Opacity_Delay}
  n313 = Divide(
   A=Subtract(A=Time(), B=Add(A=@n311, B=@n312))
   B=ScalarParameter() {defaultValue=1, group=Function_Left_Right_Decal V2_Reactive, parameterName=FDecalaV2_Opacity_Duration}
  )
  NamedRerouteDeclaration(None=Power(
    Base=Saturate(None=@n313)
    Exp=ScalarParameter() {defaultValue=1, group=Function_Left_Right_Decal V2_Reactive, parameterName=FDecalaV2_Opacity_Falloff}
   )) {name=Reactive_Opacity, nodeColor=<rgba 0.8516 0 1 1>}
  NamedRerouteDeclaration(None=@n189.0-1 mask) {name=Count_Uv_Msk, nodeColor=<rgba 0.5703 0 1 1>}
  n314 = TextureObjectParameter() {
   group=Function_Left_Right_Decal V2-Counter
   parameterName=FDecalV2_Counter__Tex
   texture=/Engine/Functions/Engine_MaterialFunctions02/ExampleContent/DebugNumberStrip
  }
  NamedRerouteDeclaration(None=@n314) {name=Tex_Counter, nodeColor=<rgba 0 0.8125 1 1>}
  n315 = PreSkinnedLocalBounds()
  n316 = Subtract(A=VertexInterpolator(VS=PreSkinnedPosition()).PS, B=@n315.Min)
  n317 = MaterialFunctionCall(Color=Subtract(A=@n315.Max, B=@n315.Max)) {materialFunction=&SplitComponents}
  n318 = Max(A=Max(A=@n317.R, B=@n317.R), B=@n317.B)
  NamedRerouteDeclaration(None=FunctionInput(Preview=Divide(A=@n316, B=@n318)) {bUsePreviewValueAsDefault=true, inputName=Calcated_Bounds, sortPriority=11}) {name=Caclcated_Bounds, nodeColor=<rgba 0 0.9297 1 1>}
  n319 = NamedRerouteUsage()
  n320 = ComponentMask(None=@n190) {g=true}
  n321 = Constant() {r=10}
  n322 = NamedRerouteUsage()
  n323 = Add(A=Ceil(None=Clamp(None=@n322) {clampMode=CMODE_ClampMin, maxDefault=0}))
  n324 = Frac(None=Divide(A=NamedRerouteUsage(), B=Power(Base=@n321, Exp=@n323) {desc="10-100,000,000"}))
  n325 = Divide(A=Floor(None=Multiply(A=@n324) {constB=100})) {constB=10}
  n326 = NamedRerouteUsage()
  n327 = Divide(A=@n326) {constB=1000000}
  n328 = Clamp(None=NamedRerouteUsage()) {clampMode=CMODE_ClampMax, maxDefault=0}
  n329 = Clamp(None=Subtract(A=Abs(None=@n328)) {desc=shift sampling to make room for decimal place}) {clampMode=CMODE_ClampMin, maxDefault=0}
  n330 = Floor(None=@n329)
  n331 = Frac(None=Multiply(
    A=Add(A=@n326, B=@n327) {constB=0.0001, desc="offset the floating point error precision error... "}
    B=Power(Base=@n321, Exp=@n330) {desc="10-100,000,000"}
   ))
  n332 = Divide(A=Floor(None=Multiply(A=@n331) {constB=10})) {constB=10}
  n333 = Ceil(None=@n329)
  n334 = AppendVector(A=Frac(None=NamedRerouteUsage()), B=LinearInterpolate(A=Add(A=@n320, B=@n325), B=Add(A=@n320, B=@n332), Alpha=Clamp(None=@n333)))
  n335 = NamedRerouteUsage()
  n336 = MaterialFunctionCall(UVs=NamedRerouteUsage(), Texture Size=Multiply(A=TextureProperty(Tex=@n335), B=NamedRerouteUsage())) {materialFunction=&ComputeMipLevel}
  n337 = TextureSample!(UVs=@n334, Tex=@n335, Level=@n336.MipLevel, Apply View MipBias) {mipValueMode=TMVM_MipLevel}
  n338 = NamedRerouteUsage()
  n339 = StaticSwitchParameter(True=NamedRerouteUsage(), False=NamedRerouteUsage()) {desc=Look To TopSide Shared, group=Function_Left_Right_Decal V2-Counter, parameterName="FDecalV2_Counter__Tex -->TopS_00B_Sharerd_Tex_Msk"}
  n340 = StaticSwitchParameter(True=NamedRerouteUsage(), False=@n339) {desc=Look To TopSide Shared, group=Function_Left_Right_Decal V2-Counter, parameterName="FDecalV2_Counter__Tex -->TopS_00A_Sharerd_Tex_Msk"}
  n341 = StaticSwitchParameter(True=NamedRerouteUsage(), False=@n340) {group=Function_Left_Right_Decal V2-Counter, parameterName="FDecalV2_Counter__Tex -->FDecalV2_Mask__Tex"}
  n342 = TextureProperty(Tex=@n341)
  n343 = NamedRerouteUsage()
  n344 = TextureSample(
   UVs=@n334
   Tex=@n341
   Level=MaterialFunctionCall(UVs=NamedRerouteUsage(), Texture Size=Multiply(A=@n342, B=@n343)) {materialFunction=&ComputeMipLevel}.MipLevel
  ) {mipValueMode=TMVM_MipLevel}
  n345 = Saturate(None=Multiply(
    A=ChannelMaskParameter(None=@n344.RGBA) {group=Function_Left_Right_Decal V2-Counter, parameterName=FDecalV2_Counter_Channel_Msk}.R
    B=NamedRerouteUsage()
   ))
  n346 = StaticSwitch(
   True=Saturate(None=Multiply(A=@n337.A, B=@n338))
   False=@n345
   Value=NamedRerouteDeclaration(None=StaticBoolParameter() {group=Function_Left_Right_Decal V2-Counter, parameterName="FDecalV2_Counter_Msk------>BC", sortPriority=0}) {name=Counter_Bc_State, nodeColor=<rgba 0 1 0.5313 1>}
  )
  NamedRerouteDeclaration(None=Saturate(None=Multiply(A=@n319, B=@n346))) {name=Counter, nodeColor=<rgba 1 0.8672 0 1>}
  n347 = ScalarParameter() {
   defaultValue=-9999999
   group=Function_Left_Right_Decal V2-Counter_Reactive
   parameterName=FDecalV2_Counter_Brightness_Time_BP
   sortPriority=212
  }
  n348 = ScalarParameter() {group=Function_Left_Right_Decal V2-Counter_Reactive, parameterName=FDecalV2_Counter_Brightness_Delay, sortPriority=211}
  n349 = ScalarParameter() {defaultValue=1, group=Function_Left_Right_Decal V2-Counter_Reactive, parameterName=FDecalV2_Counter_Brightness_Duration, sortPriority=211}
  n350 = Divide(A=Subtract(A=Time(), B=Add(A=@n347, B=@n348)), B=@n349)
  n351 = ScalarParameter() {defaultValue=1, group=Function_Left_Right_Decal V2-Counter_Reactive, parameterName=FDecalV2_Counter_Brightness_Falloff, sortPriority=211}
  NamedRerouteDeclaration(None=Power(Base=Saturate(None=@n350), Exp=@n351)) {name=reactive_Count_Brightness, nodeColor=<rgba 0 0.5078 1 1>}
  n352 = ScalarParameter() {defaultValue=1, group=Function_Left_Right_Decal V2_Emissive, parameterName=FDecalV2_Emissive_Mask_Add_Counter_Brightness, sortPriority=5}
  NamedRerouteDeclaration(None=Multiply(A=@n345, B=@n352)) {name=Counter_output, nodeColor=<rgba 0 1 0.3438 1>}
  NamedRerouteDeclaration(None=@n82) {name=Counter_State, nodeColor=<rgba 1 0.0234 0 1>}
  NamedRerouteDeclaration(None=StaticSwitchParameter(True=@n132, False=StaticBool()) {group=Function_Left_Right_Decal V2_Emissive, parameterName=FDecalV2_Emissive, sortPriority=0}) {name=Add_Counter_Stat, nodeColor=<rgba 0 0.2969 1 1>}
  NamedRerouteDeclaration(None=@n130.EmissiveColor) {name=In_Emissive, nodeColor=<rgba 0.4062 0 1 1>}
  n353 = NamedRerouteUsage()
  n354 = NamedRerouteUsage()
  n355 = NamedRerouteUsage()
  n356 = NamedRerouteUsage()
  n357 = VectorParameter() {group=Function_Left_Right_Decal V2_Uvs, parameterName=FDecalV2_UV_Warp-Speed, defaultValue=<rgba 0 0 0 1>}
  n358 = Multiply(A=Time(), B=@n357.RGB)
  n359 = StaticBool()
  n360 = NamedRerouteUsage()
  n361 = StaticSwitchParameter(True=NamedRerouteUsage(), False=NamedRerouteUsage()) {group=Function_Left_Right_Decal V2_Uvs, parameterName="FDecalV2_UV_Warp--FDecalV2_Mask_-------->TopS_00B_Sharerd_Tex_Msk"}
  n362 = ChannelMaskParameter(None=TextureSample!(
    UVs=MaterialFunctionCall!(
     Normal_Map
     Offset Rt
     Pivot
     Caculated_Bounds=@n353
     Pre_Skinned_Normals=@n354
     is_State=@n355
     UV_Rotation
     Uvs_Offset
     Uvs_TileRate=@n356
     Uvs_Offset_Pivot=@n358
     For_Safe_Zones=@n359
     Rotation_Comp
     Clamp Y
     Clamp X
     Clamp Padding_Comp
     No_Rotation=@n359
    ) {materialFunction=&/MF_Projection_Side_Normal_V2}.UVs
    Tex=StaticSwitchParameter(True=@n360, False=@n361) {group=Function_Left_Right_Decal V2_Uvs, parameterName="FDecalV2_UV_Warp--FDecalV2_Mask_-------->TopS_00A_Sharerd_Tex_Msk"}
    Apply View MipBias
   ).RGBA) {group=Function_Left_Right_Decal V2_Uvs, parameterName=FDecalV2_UV_Warp-Channel}
  n363 = ScalarParameter() {defaultValue=-0.5, group=Function_Left_Right_Decal V2_Uvs, parameterName=FDecalV2_UV_Warp-Offset}
  NamedRerouteDeclaration(None=Multiply(A=Add(A=@n362.R, B=@n363), B=ScalarParameter() {group=Function_Left_Right_Decal V2_Uvs, parameterName=FDecalV2_UV_Warp-Strength})) {name=Warp, nodeColor=<rgba 0 1 0.1797 1>}
  n364 = ScalarParameter() {defaultValue=1, group=Function_Left_Right_Decal V2_Uvs, parameterName=FDecalV2_UV_Warp_Scale_Vehicle}
  n365 = ScalarParameter() {defaultValue=1, group=Function_Left_Right_Decal V2_Uvs, parameterName=FDecalV2_UV_Warp_Scale_VehicleSide}
  n366 = MaterialFunctionCall(Float4=NamedRerouteUsage()) {materialFunction=&BreakOutFloat4Components}
  n367 = LinearInterpolate(
   A=LinearInterpolate(
    A=ScalarParameter() {defaultValue=1, group=Function_Left_Right_Decal V2_Uvs, parameterName=FDecalV2_UV_Warp_Scale_Weapon}
    B=LinearInterpolate(A=@n364, B=@n365, Alpha=@n366.G)
    Alpha=@n366.R
   )
   B=LinearInterpolate(
    A=ScalarParameter() {defaultValue=1, group=Function_Left_Right_Decal V2_Uvs, parameterName=FDecalV2_UV_Warp_Scale_Mech}
    B=ScalarParameter() {defaultValue=1, group=Function_Left_Right_Decal V2_Uvs, parameterName=FDecalV2_UV_Warp_Scale_Character}
    Alpha=@n366.A
   )
   Alpha=@n366.B
  )
  n368 = ScalarParameter() {defaultValue=1, group=Function_Left_Right_Decal V2_Uvs, parameterName=FDecalV2_UV_Warp_Scale_Backpack}
  n369 = MaterialFunctionCall(Float2=NamedRerouteUsage()) {materialFunction=&BreakOutFloat2Components}
  NamedRerouteDeclaration(None=LinearInterpolate(
    A=LinearInterpolate(A=@n367, B=@n368, Alpha=@n369.R)
    B=ScalarParameter() {defaultValue=1, group=Function_Left_Right_Decal V2_Uvs, parameterName=FDecalV2_UV_Warp_Scale_Pickaxe}
    Alpha=@n369.G
   )) {name=Warp_Uvs_Scale, nodeColor=<rgba 0 0.5547 1 1>}
  n370 = ScalarParameter() {defaultValue=1, group=Function_Left_Right_Decal V2_Emissive_Reactive, parameterName=FDecalV2_Emissive_Brightness_Time__Start}
  n371 = ScalarParameter() {group=Function_Left_Right_Decal V2_Emissive_Reactive, parameterName=FDecalV2_Emissive_Brightness_Time__End}
  n372 = NamedRerouteUsage()
  n373 = StaticSwitch(True=Add(A=@n122.R, B=@n123), False=@n122.R, Value=NamedRerouteUsage())
  n374 = ScalarParameter() {group=Function_Left_Right_Decal V2_Emissive, parameterName=FDecalV2_Emissive_Warp_Streagth}
  n375 = Add(
   A=ChannelMaskParameter(None=NamedRerouteUsage()) {group=Function_Left_Right_Decal V2_Emissive, parameterName=FDecalV2_Emissive_Warp_Mask_Channel}.R
   B=ScalarParameter() {group=Function_Left_Right_Decal V2_Emissive, parameterName=FDecalV2_Emissive_Warp_Offset}
  )
  n376 = ComponentMask(None=@n115) {g=true, r=true}
  n377 = StaticSwitchParameter(True=Add(A=Multiply(A=@n374, B=@n375), B=@n376), False=@n376) {group=Function_Left_Right_Decal V2_Emissive, parameterName=FDecalV2_Emissive_Warp_ Msk}
  n378 = LinearInterpolate(
   A=Frac(None=@n377)
   B=@n377
   Alpha=ScalarParameter() {group=Function_Left_Right_Decal V2_Emissive, parameterName="FDecalV2_Emissive_Gradent_Clamp  1=on"}
  )
  n379 = StaticSwitchParameter(True=NamedRerouteUsage(), False=NamedRerouteUsage()) {
   desc=Look To TopSide Shared
   group=Function_Left_Right_Decal V2_Emissive
   parameterName="FDecalV2_Emissive_Gradent__Tex -->TopS_00B_Sharerd_Tex_Msk"
  }
  n380 = StaticSwitchParameter(True=NamedRerouteUsage(), False=@n379) {
   desc=Look To TopSide Shared
   group=Function_Left_Right_Decal V2_Emissive
   parameterName="FDecalV2_Emissive_Gradent__Tex -->TopS_00A_Sharerd_Tex_Msk"
  }
  n381 = StaticSwitchParameter(True=NamedRerouteUsage(), False=@n380) {group=Function_Left_Right_Decal V2_Emissive, parameterName="FDecalV2_Emissive_Gradent__Tex -->FDecalV2_Mask__Tex"}
  n382 = Divide(
   A=Frac(None=ComponentMask(None=@n115) {r=true})
   B=ScalarParameter() {defaultValue=1, group=Function_Left_Right_Decal V2_Emissive, parameterName=FDecalV2_Emissive_Gradent_Thickness}
  ) {constB=0.2}
  n383 = ScalarParameter() {defaultValue=1, group=Function_Left_Right_Decal V2_Emissive, parameterName=FDecalV2_Emissive_Gradent_Falloff, sortPriority=211}
  n384 = StaticSwitchParameter(
   True=ChannelMaskParameter(None=TextureSample(UVs=@n378, Tex=@n381).RGBA) {group=Function_Left_Right_Decal V2_Emissive, parameterName=FDecalV2_Emissive_Gradent_MSk_Channel}.R
   False=Power(Base=OneMinus(None=@n382), Exp=@n383)
  ) {group=Function_Left_Right_Decal V2_Emissive, parameterName="FDecalV2_Emissive_Gradent ---> use Msk For Gradent"}
  n385 = Multiply(A=NamedRerouteUsage(), B=@n384)
  n386 = Multiply(A=@n373, B=@n385)
  n387 = Add(A=@n386, B=@n385)
  n388 = ScalarParameter() {defaultValue=1, group=Function_Left_Right_Decal V2_Emissive, parameterName=FDecalV2_Emissive_Height_Contrast}
  n389 = MaterialFunctionCall(In=Saturate(None=Multiply(A=@n386, B=@n387)), Contrast=@n388) {materialFunction=&CheapContrast}
  n390 = StaticSwitch(True=Add(A=@n384, B=@n121), False=@n384, Value=@n132)
  n391 = StaticSwitchParameter(
   True=StaticSwitch(True=Add(A=@n120.R, B=@n121), False=@n120.R, Value=@n132)
   False=StaticSwitchParameter(
    True=MaterialFunctionCall(Transition Phase=@n385, Height Texture=@n373, Contrast=@n388) {materialFunction=&HeightLerp}.Alpha
    False=StaticSwitchParameter(True=@n389.Result, False=@n390) {group=Function_Left_Right_Decal V2_Emissive, parameterName=FDecalV2_Emissive_Height___X+X, sortPriority=4}
   ) {group=Function_Left_Right_Decal V2_Emissive, parameterName=FDecalV2_Emissive_Height___Lerp, sortPriority=4}
  ) {group=Function_Left_Right_Decal V2_Emissive, parameterName=FDecalV2_Emissive_Only_Mask, sortPriority=4}
  n392 = Saturate(None=Add(
    A=@n391
    B=ScalarParameter() {defaultValue=1, group=Function_Left_Right_Decal V2_Emissive, parameterName="FDecalV2_Emissive_CC_Mask   0=ON", sliderMax=1}
   ))
  n393 = ScalarParameter() {group=Function_Left_Right_Decal V2_Emissive, parameterName=FDecalV2_Emissive_CC-Animated}
  n394 = Time()
  n395 = Multiply(A=ScalarParameter() {defaultValue=1, group=Function_Left_Right_Decal V2_Emissive, parameterName=FDecalV2_Emissive_CC_TileRate}, B=@n391)
  n396 = ScalarParameter() {group=Function_Left_Right_Decal V2_Emissive, parameterName=FDecalV2_Emissive_CC_Offset}
  n397 = Frac(None=Add(A=Multiply(A=@n393, B=@n394), B=Add(A=@n395, B=@n396)))
  n398 = CurveAtlasRowParameter(CurveTime=@n397) {
   atlas=&CruveAtlas_Wraps_01
   curve=/Game/Weapons/WeaponSkins/Wraps/Textures/C_GradSimple
   group=Function_Left_Right_Decal V2_Emissive
   parameterName=FDecalV2_Emissive_CC
  }
  n399 = MaterialFunctionCall(
   Index=ScalarParameter() {group=Function_Left_Right_Decal V2_Emissive, parameterName="FDecalV2_Emissive_CC  Index", sliderMax=1}
   CCinput=@n398
   CCinput_Alpha=@n398
   UV=@n397
  ) {materialFunction=&MF_Global_CC_Tex}
  n400 = ScalarParameter() {defaultValue=1, group=Function_Left_Right_Decal V2_Emissive, parameterName=FDecalV2_Emissive_CC_Brightness}
  n401 = ScalarParameter() {defaultValue=1, group=Function_Left_Right_Decal V2_Emissive, parameterName="FDecalV2_Emissive_CC_Brightness USE CCAlpha 1=On", sliderMax=1}
  n402 = Multiply(A=@n399.3V, B=LinearInterpolate(A=@n400, B=Multiply(A=@n399.Alpha, B=@n400), Alpha=Ceil(None=@n401)))
  n403 = VectorParameter() {group=Function_Left_Right_Decal V2_Emissive, parameterName="FDecalV2-Emissive-Color--Alpha=Multplyer", defaultValue=<rgba 1 1 1 1>}
  n404 = Multiply(A=@n391, B=@n403.RGB)
  n405 = Multiply(
   A=StaticSwitchParameter(True=Multiply(A=@n392, B=@n402), False=Multiply(A=@n404, B=@n403.A)) {group=Function_Left_Right_Decal V2_Emissive, parameterName="FDecalV2-Emissive-Color--->CCurve"}
   B=NamedRerouteUsage()
  )
  n406 = StaticSwitchParameter(True=Multiply(A=@n405, B=LinearInterpolate(A=@n370, B=@n371, Alpha=@n372)), False=@n405) {group=Function_Left_Right_Decal V2_Emissive_Reactive, parameterName=FDecalV2-Emissive_Reactive_Brightness---------------------}
  n407 = Multiply(A=NamedRerouteUsage(), B=@n406)
  n408 = NamedRerouteUsage()
  NamedRerouteDeclaration(None=Multiply(A=StaticSwitch(True=@n407, False=@n406, Value=@n408), B=NamedRerouteUsage())) {name=Emissive, nodeColor=<rgba 0.2656 1 0 1>}
  n409 = Constant() {r=1}
  n410 = ScalarParameter() {defaultValue=0.6, group=Function_Left_Right_Decal V2_Emissive, parameterName=FDecalV2_Emissive_BC_Alpha_Clamp, sortPriority=0}
  n411 = Step(Y=@n410, X=NamedRerouteUsage())
  n412 = StaticSwitchParameter(True=@n409, False=@n411) {group=Function_Left_Right_Decal V2_Emissive, parameterName=FDecalV2_Emissive_Height___X+X, sortPriority=5}
  NamedRerouteDeclaration(None=Multiply(
    A=NamedRerouteUsage()
    B=StaticSwitchParameter(True=@n409, False=@n412) {group=Function_Left_Right_Decal V2_Emissive, parameterName=FDecalV2_Emissive_Height___Lerp, sortPriority=5}
   )) {name=Emissive_Msk_Simple, nodeColor=<rgba 0 1 0.1094 1>}
  NamedRerouteDeclaration(None=TextureObjectParameter() {group=Function_Left_Right_Decal V2, parameterName=FDecalV2_Mask__Tex, texture=&T_BlacksRGB}) {name=Tex_Sample_MSk, nodeColor=<rgba 0.1016 1 0 1>}
  n413 = Constant() {r=0.5}
  n414 = VectorParameter() {group=Function_Left_Right_Decal V2_Emissive, parameterName=FDecalV2_Emissive_Msk_Offset, sortPriority=5}
  n415 = MaterialFunctionCall(Float2=Add(A=ComponentMask(None=@n414.RGB) {g=true, r=true}, B=@n103.Rotated Values) {constB=0}) {materialFunction=&BreakOutFloat2Components}
  n416 = VectorParameter() {
   group=Function_Left_Right_Decal V2_Emissive
   parameterName=FDecalV2_Emissive_Msk_Scale
   sortPriority=5
   channelNames {a=Y Falloff, b=X Falloff, g=Y Size, r=X Size}
   defaultValue=<rgba 0.5 0.5 1 1>
  }
  n417 = SphereMask(A=@n413, B=@n415.R, Radius=@n416.Y Size, Hardness=@n416.Y Size) {attenuationRadius=0.5}
  n418 = SphereMask(A=@n415.G, B=@n413, Radius=@n416.X Size, Hardness=@n416.X Size) {attenuationRadius=0.5}
  NamedRerouteDeclaration(None=Power(Base=Multiply(A=@n417, B=@n418))) {name=Decal_Msking, nodeColor=<rgba 0 1 0.5078 1>}
  n419 = StaticSwitchParameter(
   True=LinearInterpolate(
    A=ScalarParameter() {group=Function_Left_Right_Decal V2_Emissive_Reactive, parameterName=FDecalV2_Emissive_Brightness_DarkenBase_Time_Start}
    B=ScalarParameter() {group=Function_Left_Right_Decal V2_Emissive_Reactive, parameterName=FDecalV2_Emissive_Brightness_DarkenBase_Time_End}
    Alpha=NamedRerouteUsage()
   )
   False=ScalarParameter() {group=Function_Left_Right_Decal V2_Emissive, parameterName=FDecalV2_Emissive_DarkenBase-------, sortPriority=4}
  ) {group=Function_Left_Right_Decal V2_Emissive_Reactive, parameterName=FDecalV2-Emissive_Reactive_Brightness---------------------}
  n420 = NamedRerouteUsage()
  n421 = Multiply(A=Multiply(A=@n419, B=@n391), B=@n420)
  n422 = Add(A=@n421, B=NamedRerouteUsage())
  n423 = ScalarParameter() {group=Function_Left_Right_Decal V2_Emissive, parameterName="FDecalV2_Emissive_DarkenBase-CCAlpha 1=Use", sliderMax=1, sortPriority=4}
  n424 = Ceil(None=@n423)
  n425 = StaticSwitchParameter(
   True=Multiply(A=@n419, B=@n420)
   False=StaticSwitchParameter(True=LinearInterpolate(A=@n421, B=@n422, Alpha=@n424), False=@n421) {group=Function_Left_Right_Decal V2_Emissive, parameterName="FDecalV2-Emissive-Color--->CCurve"}
  ) {
   desc=uses the Mask texture as a diffues emissive
   group=Function_Left_Right_Decal V2_Emissive
   parameterName="FDecalV2-Emissive--BC-Msk--->Simple-->Msk_Tex"
   sortPriority=500
  }
  n426 = Multiply(A=@n425, B=NamedRerouteUsage())
  n427 = NamedRerouteUsage()
  n428 = ScalarParameter() {
   defaultValue=1
   group=Function_Left_Right_Decal V2_Emissive
   parameterName="FDecalV2_Emissive_DrawBehind BC 0=On"
   sliderMax=1
   sortPriority=4
  }
  n429 = Saturate(None=Add(A=OneMinus(None=@n427), B=@n428))
  n430 = LinearInterpolate(B=OneMinus(None=Saturate(None=@n426)), Alpha=@n429) {constA=1}
  n431 = NamedRerouteUsage()
  NamedRerouteDeclaration(None=StaticSwitch(True=LinearInterpolate(B=@n430, Alpha=@n431) {constA=1}, False=@n430, Value=NamedRerouteUsage())) {name=Emissive_Darken_Base, nodeColor=<rgba 0 0.1328 1 1>}
  n432 = ScalarParameter() {
   defaultValue=-9999999
   group=Function_Left_Right_Decal V2_Emissive_Reactive
   parameterName=FDecalV2_Emissive_Gradent_Time_BP
   sortPriority=212
  }
  n433 = ScalarParameter() {group=Function_Left_Right_Decal V2_Emissive_Reactive, parameterName=FDecalV2_Emissive_Gradent_Delay, sortPriority=211}
  n434 = ScalarParameter() {defaultValue=0.25, group=Function_Left_Right_Decal V2_Emissive_Reactive, parameterName=FDecalV2_Emissive_Gradent_Duration, sortPriority=211}
  n435 = Divide(A=Subtract(A=Time(), B=Add(A=@n432, B=@n433)), B=@n434)
  n436 = ScalarParameter() {defaultValue=1, group=Function_Left_Right_Decal V2_Emissive_Reactive, parameterName=FDecalV2_Emissive_Gradent_Falloff, sortPriority=211}
  NamedRerouteDeclaration(None=Power(Base=Saturate(None=@n435), Exp=@n436)) {name=reactive_Gradent, nodeColor=<rgba 0 0.5078 1 1>}
  NamedRerouteDeclaration(None=@n411) {name=Alpha_Msk, nodeColor=<rgba 1 0 0.9375 1>}
  n437 = ScalarParameter() {
   defaultValue=-9999999
   group=Function_Left_Right_Decal V2_Emissive_Reactive
   parameterName=FDecalV2_Emissive_Brightness_Time_BP
   sortPriority=50
  }
  n438 = ScalarParameter() {group=Function_Left_Right_Decal V2_Emissive_Reactive, parameterName=FDecalV2_Emissive_Brightness_Delay}
  n439 = Divide(
   A=Subtract(A=Time(), B=Add(A=@n437, B=@n438))
   B=ScalarParameter() {defaultValue=0.25, group=Function_Left_Right_Decal V2_Emissive_Reactive, parameterName=FDecalV2_Emissive_Brightness_Duration}
  )
  NamedRerouteDeclaration(None=Power(
    Base=Saturate(None=@n439)
    Exp=ScalarParameter() {defaultValue=1, group=Function_Left_Right_Decal V2_Emissive_Reactive, parameterName=FDecalV2_Emissive_Brightness_Falloff}
   )) {name=Reactive_Brightness, nodeColor=<rgba 0.8516 0 1 1>}
  NamedRerouteDeclaration(None=@n429) {name=Alpha_MSk_B, nodeColor=<rgba 1 0 0.8203 1>}
  NamedRerouteDeclaration(None=@n399.Alpha) {name=CC_Alpha, nodeColor=<rgba 0 1 0.25 1>}
  NamedRerouteDeclaration(None=@n213) {name=UVs_Raw_simi, nodeColor=<rgba 0.125 0 1 1>}
  n440 = TextureObjectParameter() {
   group=Function_Left_Right_Decal V2_Emissive
   parameterName=FDecalV2_Emissive_Gradent__Tex
   texture=&T_BlacksRGB
  }
  NamedRerouteDeclaration(None=@n440) {name=Tex_Sample_Grad, nodeColor=<rgba 0.1016 1 0 1>}
  NamedRerouteDeclaration(None=@n337.RGB) {name=Counter_BC, nodeColor=<rgba 0.8516 0 1 1>}
  n441 = PreSkinnedNormal()
  NamedRerouteDeclaration(None=FunctionInput(Preview=VertexInterpolator(VS=@n441).PS) {bUsePreviewValueAsDefault=true, inputName=Pre-Skinned Local Normal, sortPriority=10}) {name=PreSkin_Normal, nodeColor=<rgba 0.8281 1 0 1>}
  n442 = TextureObjectParameter() {
   group=Function_Left_Right_Decal V2-Counter
   parameterName=FDecalV2_Counter_N__Tex
   samplerType=SAMPLERTYPE_Normal
   texture=&FlatNormalBlue
  }
  NamedRerouteDeclaration(None=@n442) {name=Tex_Counter_N, nodeColor=<rgba 0 0.8125 1 1>}
  n443 = NamedRerouteUsage()
  n444 = StaticBool() {value=true}
  NamedRerouteDeclaration(None=MaterialFunctionCall(
    Normals_Source=TextureSample!(UVs=@n334, Tex=@n443, Level=@n336.MipLevel, Apply View MipBias) {mipValueMode=TMVM_MipLevel}.RGB
    Normals_Source_Use=@n444
    UVs=NamedRerouteUsage()
    UV_Rotation=NamedRerouteUsage()
    Pre_Skinned_Normals=NamedRerouteUsage()
    is_State=NamedRerouteUsage()
    "Left\\Right"=@n444
   ) {materialFunction=&MF_UVs_Side_Fix_Normals}.Raw_Normal) {name=Counter N, nodeColor=<rgba 0.7578 1 0 1>}
  n445 = TextureObjectParameter() {group=Function_Left_Right_Decal V2, parameterName=FDecalV2_BaseColor_TOD_Tex, texture=&T_BlacksRGB}
  NamedRerouteDeclaration(None=@n445) {name=Tex_Sample_BC_TOD, nodeColor=<rgba 0.1016 1 0 1>}
  NamedRerouteDeclaration(None=StaticSwitch(
    True=@n93.R
    False=@n97.Result
    Value=StaticBoolParameter() {group=Function_Left_Right_Decal V2_Reactive, parameterName="FDecalV2_BC_TOD--->CCTOD", sortPriority=500}
   )) {name=TOD_BC_CCTOD_State, nodeColor=<rgba 0 1 0.8125 1>}
  NamedRerouteDeclaration(None=StaticBoolParameter() {group=Function_Left_Right_Decal V2_Reactive, parameterName=FDecalV2_BC_TOD-------------, sortPriority=500}) {name=TOD_BC_State, nodeColor=<rgba 0 1 0.8125 1>}
  n446 = TextureSample!(UVs=@n231.UVs, Tex=NamedRerouteUsage(), "DDX(UVs)"=@n232, "DDY(UVs)"=@n233, Apply View MipBias) {mipValueMode=TMVM_Derivative}
  NamedRerouteDeclaration(None=@n446.RGB) {name=Flipbook_RGB_B, nodeColor=<rgba 0 0.0156 1 1>}
  NamedRerouteDeclaration(None=@n446.A) {name=Flipbook_Alpha_B, nodeColor=<rgba 0 0.0156 1 1>}
  NamedRerouteDeclaration(None=@n446.RGBA) {name=Flipbook RGBA_B, nodeColor=<rgba 0 0.0156 1 1>}
  NamedRerouteDeclaration(None=StaticSwitch(True=NamedRerouteUsage(), False=@n296.RGB, Value=@n297)) {name=BC_TOD_Output, nodeColor=<rgba 0 0.7891 1 1>}
  NamedRerouteDeclaration(None=@n130.Normal) {name=In_Normal, nodeColor=<rgba 0.4062 0 1 1>}
  NamedRerouteDeclaration(None=StaticBoolParameter() {group=Function_Left_Right_Decal V2_Reactive, parameterName=FDecalV2_BCEmissive_TOD-------------, sortPriority=500}) {name=TOD_BC_State_1, nodeColor=<rgba 0 1 0.8125 1>}
  NamedRerouteDeclaration(None=StaticSwitch(
    True=@n93.R
    False=@n97.Result
    Value=StaticBoolParameter() {group=Function_Left_Right_Decal V2_Reactive, parameterName="FDecalV2_BCEmissive_TOD--->CCTOD", sortPriority=500}
   )) {name=TOD_BC_CCTOD_State_1, nodeColor=<rgba 0 1 0.8125 1>}
  n447 = Time()
  n448 = ScalarParameter() {defaultValue=-9999999, group=Function_Left_Right_Decal V2_Reactive, parameterName=FDecalV2_Scale_Time_BP, sortPriority=212}
  n449 = Add(
   A=@n448
   B=LinearInterpolate(
    A=ScalarParameter() {group=Function_Left_Right_Decal V2_Reactive, parameterName=FDecalV2_Scale ----------------}
    B=ScalarParameter() {group=Function_Left_Right_Decal V2_Reactive, parameterName=FDecalV2_Scale_Delay, sortPriority=211}
   ) {constAlpha=1}
  )
  n450 = Saturate(None=Divide(
    A=Subtract(A=@n447, B=@n449)
    B=ScalarParameter() {defaultValue=1, group=Function_Left_Right_Decal V2_Reactive, parameterName=FDecalV2_Scale_Duration, sortPriority=211}
   ))
  n451 = ScalarParameter() {defaultValue=1, group=Function_Left_Right_Decal V2_Reactive, parameterName=FDecalV2_Scale_Falloff, sortPriority=211}
  NamedRerouteDeclaration(None=LinearInterpolate(
    A=ScalarParameter() {defaultValue=1, group=Function_Left_Right_Decal V2_Reactive, parameterName=FDecalV2_Scale Start}
    B=ScalarParameter() {defaultValue=1, group=Function_Left_Right_Decal V2_Reactive, parameterName=FDecalV2_Scale End}
    Alpha=Power(Base=@n450, Exp=@n451)
   )) {name=reactive_Scale, nodeColor=<rgba 0 0.5078 1 1>}
  n452 = ScalarParameter() {defaultValue=2, group=Function_Left_Right_Decal V2-Split, parameterName=FDecalV2_Split_L/R_Number_Rows}
  n453 = ScalarParameter() {defaultValue=2, group=Function_Left_Right_Decal V2-Split, parameterName=FDecalV2_Split_L/R_Number_Columns}
  n454 = Multiply(A=@n452, B=@n453)
  n455 = Saturate(None=NamedRerouteUsage())
  n456 = StaticBool()
  n457 = MaterialFunctionCall(
   "Animation  Phase (0-1)"=Divide(A=ScalarParameter() {group=Function_Left_Right_Decal V2-Split, parameterName=FDecalV2_Split_Left_Frame}, B=@n454)
   Number of Rows=@n452
   Number of Columns=@n453
   UVs=@n455
   "Clamp Anim (see tooltip)"=@n456
  ) {materialFunction=&FlipBook}
  n458 = NamedRerouteUsage()
  n459 = DDX(Value=@n458)
  n460 = DDY(Value=@n458)
  n461 = NamedRerouteUsage()
  n462 = MaterialFunctionCall(
   "Animation  Phase (0-1)"=Divide(A=ScalarParameter() {defaultValue=1, group=Function_Left_Right_Decal V2-Split, parameterName=FDecalV2_Split_Right_Frame}, B=@n454)
   Number of Rows=@n452
   Number of Columns=@n453
   UVs=@n455
   "Clamp Anim (see tooltip)"=@n456
  ) {materialFunction=&FlipBook}
  n463 = MaterialFunctionCall(Color=@n1) {materialFunction=&SplitComponents}
  n464 = Multiply(A=@n463.G) {constB=-1}
  n465 = MaterialFunctionCall(Color=NamedRerouteUsage()) {materialFunction=&SplitComponents}
  n466 = LinearInterpolate(A=@n463.R, B=LinearInterpolate(A=Multiply(A=@n463.B) {constB=-1}, B=@n464, Alpha=@n465.G), Alpha=@n465.R)
  n467 = Saturate(None=Ceil(None=LinearInterpolate(A=@n466, B=@n464, Alpha=@n465.B)))
  NamedRerouteDeclaration(None=LinearInterpolate(
    A=TextureSample(UVs=@n457.UVs, Tex=@n461, "DDX(UVs)"=@n459, "DDY(UVs)"=@n460) {mipValueMode=TMVM_Derivative}.RGBA
    B=TextureSample(UVs=@n462.UVs, Tex=@n461, "DDX(UVs)"=@n459, "DDY(UVs)"=@n460) {mipValueMode=TMVM_Derivative}.RGBA
    Alpha=@n467
   )) {name=Split_Bc, nodeColor=<rgba 0.9922 0 1 1>}
  NamedRerouteDeclaration(None=StaticBoolParameter() {group=Function_Left_Right_Decal V2, parameterName="FDecalV2_Left\\Right_Split_FlipBook", sortPriority=0}) {name="L\\R_Split_State", nodeColor=<rgba 0 1 0.6719 1>}
  n468 = StaticSwitchParameter(True=NamedRerouteUsage(), False=NamedRerouteUsage()) {desc=Look To TopSide Shared, group=Function_Left_Right_Decal V2-Split, parameterName="FDecalV2_Split_Mask__Tex -->TopS_00B_Sharerd_Tex_Msk"}
  n469 = StaticSwitchParameter(True=NamedRerouteUsage(), False=@n468) {desc=Look To TopSide Shared, group=Function_Left_Right_Decal V2-Split, parameterName="FDecalV2_Split_Mask__Tex -->TopS_00A_Sharerd_Tex_Msk"}
  n470 = NamedRerouteUsage()
  n471 = DDX(Value=@n470)
  n472 = DDY(Value=@n470)
  NamedRerouteDeclaration(None=LinearInterpolate(
    A=TextureSample(UVs=@n457.UVs, Tex=@n469, "DDX(UVs)"=@n471, "DDY(UVs)"=@n472) {mipValueMode=TMVM_Derivative}.RGBA
    B=TextureSample(UVs=@n462.UVs, Tex=@n469, "DDX(UVs)"=@n471, "DDY(UVs)"=@n472) {mipValueMode=TMVM_Derivative}.RGBA
    Alpha=@n467
   )) {name=Split_Msk, nodeColor=<rgba 0.9922 0 1 1>}
  n473 = NamedRerouteUsage()
  n474 = NamedRerouteUsage()
  n475 = DDX(Value=@n474)
  n476 = DDY(Value=@n474)
  n477 = TextureSample(UVs=@n457.UVs, Tex=@n473, "DDX(UVs)"=@n475, "DDY(UVs)"=@n476) {mipValueMode=TMVM_Derivative}
  n478 = TextureSample(UVs=@n462.UVs, Tex=@n473, "DDX(UVs)"=@n475, "DDY(UVs)"=@n476) {mipValueMode=TMVM_Derivative}
  NamedRerouteDeclaration(None=MaterialFunctionCall(
    Normals_Source=LinearInterpolate(A=@n477.RGB, B=@n478.RGB, Alpha=@n467)
    Normals_Source_Use=StaticBool() {value=true}
    UV_Rotation=NamedRerouteUsage()
    Pre_Skinned_Normals=NamedRerouteUsage()
    is_State=NamedRerouteUsage()
    "Left\\Right"=NamedRerouteUsage()
   ) {materialFunction=&MF_UVs_Side_Fix_Normals}.Raw_Normal) {name=Split_Normal, nodeColor=<rgba 0.9922 0 1 1>}
  n479 = ScalarParameter() {
   defaultValue=182
   group=Function_Left_Right_Decal V2_Emissive_Reactive_TOD
   parameterName="FDecalV2_Emissive_TOD_CC  Index"
   sortPriority=500
  }
  n480 = ScalarParameter() {group=00-FrontEnd-Overrides, parameterName=00-FE-TODCC-Override}
  n481 = Add(
   A=@n480
   B=CollectionParameter() {collection=&FortniteMaterialParameters, parameterName=TimeOfDay_0-1}
  )
  n482 = CurveAtlasRowParameter(CurveTime=@n481) {
   atlas=&CruveAtlas_Wraps_01
   curve=&Curve_TimeOfDay
   defaultValue=182
   group=Function_Left_Right_Decal V2_Emissive_Reactive_TOD
   parameterName=FDecalV2_Emissive_TOD_CC
   sortPriority=500
  }
  n483 = MaterialFunctionCall(Index=@n479, CCinput=@n482, CCinput_Alpha=@n482, UV=@n481) {materialFunction=&MF_Global_CC_Tex}
  n484 = ScalarParameter() {defaultValue=1, group=Function_Left_Right_Decal V2_Emissive_Reactive_TOD, parameterName=FDecalV2_Emissive_TOD_Morrning, sortPriority=500}
  n485 = ScalarParameter() {defaultValue=1, group=Function_Left_Right_Decal V2_Emissive_Reactive_TOD, parameterName=FDecalV2_Emissive_TOD_Day, sortPriority=500}
  n486 = Add(A=LinearInterpolate(A=@n480, B=Multiply(A=@n480) {constB=-1}, Alpha=@n485), B=@n485)
  n487 = ScalarParameter() {defaultValue=1, group=Function_Left_Right_Decal V2_Emissive_Reactive_TOD, parameterName=FDecalV2_Emissive_TOD_Evening, sortPriority=500}
  n488 = ScalarParameter() {defaultValue=1, group=Function_Left_Right_Decal V2_Emissive_Reactive_TOD, parameterName=FDecalV2_Emissive_TOD_Night, sortPriority=500}
  NamedRerouteDeclaration(None=StaticSwitch(
    True=ChannelMaskParameter(None=@n483.4v) {group=Function_Left_Right_Decal V2_Emissive_Reactive_TOD, parameterName=FDecalV2_Emissive_TOD_CC_Channel, sortPriority=500}.R
    False=MaterialFunctionCall(Value_Morning=@n484, Value_Day=@n486, Value_Evening=@n487, Value_Night=@n488) {materialFunction=&Scalar_Time_of_Day}.Result
    Value=StaticBoolParameter() {group=Function_Left_Right_Decal V2_Emissive_Reactive_TOD, parameterName="FDecalV2_Emissive_TOD--->CCTOD", sortPriority=500}
   )) {name=TOD_Emissive, nodeColor=<rgba 0.875 0 1 1>}
  NamedRerouteDeclaration(None=StaticBoolParameter() {group=Function_Left_Right_Decal V2_Emissive_Reactive, parameterName=FDecalV2_Emissive_TOD, sortPriority=500}) {name=TOD_Emissive_State, nodeColor=<rgba 0.875 0 1 1>}
  n489 = TextureObjectParameter() {group=Function_Left_Right_Decal V2, parameterName=FDecalV2_Base_RGBA_Tex, texture=&T_BlacksRGB}
  NamedRerouteDeclaration(None=@n489) {name=Tex_Sample_RGBA, nodeColor=<rgba 0.1016 1 0 1>}
  NamedRerouteUsage()
  NamedRerouteUsage()
  NamedRerouteUsage()
  NamedRerouteUsage()
  NamedRerouteUsage()
  NamedRerouteUsage()
  NamedRerouteUsage()
  NamedRerouteUsage()
  Round()
  ScalarParameter() {defaultValue=182, group=Function_Left_Right_Decal V2_Reactive, parameterName="FDecalV2_Opacity_TOD_CC  Index", sortPriority=500}
  StaticBool()
  StaticSwitchParameter(True=Multiply(A=NamedRerouteUsage(), B=NamedRerouteUsage()), False=@n405) {
   desc=uses the Mask texture as a diffues emissive
   group=Function_Left_Right_Decal V2_Emissive
   parameterName="FDecalV2-Emissive--BC-Msk--->Simple-->Msk_Tex"
   sortPriority=500
  }
  VectorParameter() {group="Function_Left_Right_Decal  V2 msk", parameterName="FDecalV2_Mask_RGBA_Output_TS01<--->TS02 Fill Value", defaultValue=<rgba 1 1 1 1>}
 ]}
