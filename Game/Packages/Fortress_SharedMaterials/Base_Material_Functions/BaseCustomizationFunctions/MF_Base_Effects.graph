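# Material function that stacks the wrap customization effects onto a base material, with state inputs for vehicle, mech, character, backpack and pickaxe.
# Effects run in series from simple customization through decals, comic, liquid metal, gemstone, ice and others, ending in hue shift and audio pulse.
# Named reroutes share masks, bounds and safe zones; decal and hot wax masks are also output.
# Metal LUT, Toon and Iridescent calls feed no output.
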MaterialFunction /Game/Packages/Fortress_SharedMaterials/Base_Material_Functions/BaseCustomizationFunctions/MF_Base_Effects
graph {nodes [
  Constant()
  n1 = MaterialFunctionCall!(
   Material_Root=NamedRerouteUsage()
   Material=NamedRerouteUsage()
   "Msk_Ts01<--->Ts02"=NamedRerouteUsage()
   IsState_Bv4=NamedRerouteUsage()
   CM_Msk=NamedRerouteUsage()
   IsState=NamedRerouteUsage()
   Pre-Skinned Local Normal=NamedRerouteUsage()
   Calcated_Bounds=NamedRerouteUsage()
  ) {materialFunction=/Game/Weapons/WeaponSkins/Wraps/MaterialFunctions/MF_Customization_Simple_V2}
  n2 = NamedRerouteUsage()
  n3 = MaterialFunctionCall!(
   Root_Matieral=NamedRerouteUsage()
   Material=MaterialFunctionCall!(
    Material=MaterialFunctionCall!(Material=@n1.Result, "Customization Mask (_CM B)"=@n2) {materialFunction=/Game/Weapons/WeaponSkins/Wraps/MaterialFunctions/MF_CritterManiac_Aged_FilmGrain_Wrap}.Result
    IsState_B=NamedRerouteUsage()
    IsState=NamedRerouteUsage()
    Pre-Skinned Local Normal=NamedRerouteUsage()
    Calcated_Bounds=NamedRerouteUsage()
    Mask Crevise=NamedRerouteUsage()
    Mask Secondary=NamedRerouteUsage()
    Mask Primary=NamedRerouteUsage()
    TS_Mask_01=NamedRerouteUsage()
    TS_Mask_02=NamedRerouteUsage()
   ) {materialFunction=/Game/Weapons/WeaponSkins/Wraps/MaterialFunctions/MF_Effect_Multicolor}.Result
   "Msk_Ts01<--->Ts02"=NamedRerouteUsage()
   IsState_B=NamedRerouteUsage()
   IsState=NamedRerouteUsage()
   CM_Msk=NamedRerouteUsage()
   Pre-Skinned Local Normal=NamedRerouteUsage()
   Calcated_Bounds=NamedRerouteUsage()
   Object-SafeZone-Back=NamedRerouteUsage()
   Object-SafeZone-Middle=NamedRerouteUsage()
   Object-SafeZone-Front=NamedRerouteUsage()
  ) {materialFunction=/Game/Weapons/WeaponSkins/Wraps/MaterialFunctions/MF_Effect_Decal_Reactive}
  n4 = MaterialFunctionCall!(
   Material_Root=NamedRerouteUsage()
   Material=@n3.Result
   "Msk_Ts01<--->Ts02"=NamedRerouteUsage()
   ForwardAxis=NamedRerouteUsage()
   IsState=NamedRerouteUsage()
   CM_Msk=NamedRerouteUsage()
   IsState_BV4=NamedRerouteUsage()
   Pre-Skinned Local Normal=NamedRerouteUsage()
   Calcated_Bounds=NamedRerouteUsage()
   Object-SafeZone-Back=NamedRerouteUsage()
   Object-SafeZone-Middle=NamedRerouteUsage()
   Object-SafeZone-Front=NamedRerouteUsage()
  ) {materialFunction=/Game/Weapons/WeaponSkins/Wraps/MaterialFunctions/MF_Function_Left_Right_Decal_V2}
  n5 = MaterialFunctionCall!(
   Material=@n4.Result
   IsState=NamedRerouteUsage()
   IsState_B=NamedRerouteUsage()
   CM_Msk=NamedRerouteUsage()
   Decal_Mask=@n4.Result
   Pre-Skinned Local Normal=NamedRerouteUsage()
   Calcated_Bounds=NamedRerouteUsage()
   Object-SafeZone-VP-Character-PreView-Switch
   Object-SafeZone-Virtual-Pivot
  ) {materialFunction=/Game/Weapons/WeaponSkins/Wraps/MaterialFunctions/MF_Effect_Comic}
  n6 = NamedRerouteUsage()
  n7 = MaterialFunctionCall!(
   Material=MaterialFunctionCall!(
    In=MaterialFunctionCall!(In=@n5.Result, "Customization Mask (_CM B)"=@n6) {materialFunction=/Game/Weapons/WeaponSkins/Wraps/MaterialFunctions/MF_Effect_Reactive_Emissive}.Result
    IsState=NamedRerouteUsage()
    IsState_B=NamedRerouteUsage()
    Pre-Skinned Local Normal=NamedRerouteUsage()
    Calcated_Bounds=NamedRerouteUsage()
    Mask Crevise=NamedRerouteUsage()
    Mask Secondary=NamedRerouteUsage()
    Mask Primary=NamedRerouteUsage()
   ) {materialFunction=/Game/Weapons/WeaponSkins/Wraps/MaterialFunctions/MF_Effect_Visor}.Result
   IsState_B=NamedRerouteUsage()
   IsState=NamedRerouteUsage()
   Pre-Skinned Local Normal=NamedRerouteUsage()
   Calcated_Bounds=NamedRerouteUsage()
   Mask Crevise=NamedRerouteUsage()
   Mask Secondary=NamedRerouteUsage()
   Mask Gmap=NamedRerouteUsage()
   Mask Primary=NamedRerouteUsage()
   TS_Mask_02=NamedRerouteUsage()
   TS_Mask_01=NamedRerouteUsage()
  ) {materialFunction=/Game/Weapons/WeaponSkins/Wraps/MaterialFunctions/MF_EffectLiquidMetal}
  n8 = MaterialFunctionCall!(
   Material=MaterialFunctionCall!(
    Material_1=@n7.Result
    IsState_B=NamedRerouteUsage()
    IsState=NamedRerouteUsage()
    CM_Msk=NamedRerouteUsage()
    Pre-Skinned Local Normal=NamedRerouteUsage()
    Calcated_Bounds=NamedRerouteUsage()
    Decal_Mask_RGBA=NamedRerouteUsage()
    Decal_Mask=NamedRerouteUsage()
    TS_Layer01_Mask=NamedRerouteUsage()
    TS_Layer02_Mask=NamedRerouteUsage()
    Object-SafeZone-VP-Character-PreView-Switch=NamedRerouteUsage()
    Object-SafeZone-Virtual-Pivot=NamedRerouteUsage()
   ) {materialFunction=/Game/Weapons/WeaponSkins/Wraps/MaterialFunctions/MF_Effect_GemStone}.Result_1
   Material_Root=NamedRerouteUsage()
   IsState_B=NamedRerouteUsage()
   CM_Msk=NamedRerouteUsage()
   IsState=NamedRerouteUsage()
   Pre-Skinned Local Normal=NamedRerouteUsage()
   Calcated_Bounds=NamedRerouteUsage()
   Decal_Mask=NamedRerouteUsage()
   Decal_Mask_RGBA=NamedRerouteUsage()
  ) {materialFunction=/Game/Weapons/WeaponSkins/Wraps/MaterialFunctions/MF_Effect_HotWax}
  n9 = NamedRerouteUsage()
  n10 = MaterialFunctionCall!(
   Material=MaterialFunctionCall!(
    multi_Decal=NamedRerouteUsage()
    Material=MaterialFunctionCall!(Material=@n8.Result, CM_Msk=@n9) {materialFunction=/Game/Weapons/WeaponSkins/Wraps/MaterialFunctions/MF_Effect_BaseToEmissive}.Result
    IsState_B=NamedRerouteUsage()
    CM_Msk=@n9
    IsState=NamedRerouteUsage()
    Pre-Skinned Local Normal=NamedRerouteUsage()
    Calcated_Bounds=NamedRerouteUsage()
   ) {materialFunction=/Game/Weapons/WeaponSkins/Wraps/MaterialFunctions/MF_Effect_Sheen}.Result_1
   IsState=NamedRerouteUsage()
   IsState_B=NamedRerouteUsage()
   Pre-Skinned Local Normal=NamedRerouteUsage()
   Calcated_Bounds=NamedRerouteUsage()
   Mask Crevise=NamedRerouteUsage()
   Mask Secondary=NamedRerouteUsage()
   Mask Primary=NamedRerouteUsage()
  ) {materialFunction=/Game/Weapons/WeaponSkins/Wraps/MaterialFunctions/MF_Effect_AnimeShading}
  n11 = NamedRerouteUsage()
  n12 = NamedRerouteUsage()
  n13 = NamedRerouteUsage()
  n14 = NamedRerouteUsage()
  n15 = NamedRerouteUsage()
  n16 = MaterialFunctionCall!(
   MA_In=MaterialFunctionCall!(
    Material_Root=NamedRerouteUsage()
    Multi Decal=NamedRerouteUsage()
    InMaterial=MaterialFunctionCall!(Material=@n10.accent, IsState=@n11, IsState_B=@n12, CM_Msk=@n13, Pre-Skinned Local Normal=@n14, Calcated_Bounds=@n15, LUT Multiplier) {materialFunction=/Game/Weapons/WeaponSkins/Wraps/MaterialFunctions/MF_Effect_HatchBack_ComicS}.Result
    "Msk_Ts01<--->Ts02"=NamedRerouteUsage()
    CM_Msk=NamedRerouteUsage()
    IsState_B=NamedRerouteUsage()
    IsState=NamedRerouteUsage()
    Pre-Skinned Local Normal=NamedRerouteUsage()
    Calcated_Bounds=NamedRerouteUsage()
    Ts_Mask_02=NamedRerouteUsage()
    Ts_Mask_01=NamedRerouteUsage()
    Decal_RGBA=NamedRerouteUsage()
    Decal_Mask=NamedRerouteUsage()
   ) {materialFunction=/Game/Weapons/WeaponSkins/Wraps/MaterialFunctions/MF_Effect_Ice}.Result_1
   IsState_B=NamedRerouteUsage()
   CM_Msk=NamedRerouteUsage()
   IsState=NamedRerouteUsage()
   Pre-Skinned Local Normal=NamedRerouteUsage()
   Calcated_Bounds=NamedRerouteUsage()
  ) {materialFunction=/Game/Weapons/WeaponSkins/Wraps/MaterialFunctions/MF_Effect_Stylise}
  n17 = NamedRerouteUsage()
  FunctionOutput(None=MaterialFunctionCall!(
    Material=MaterialFunctionCall!(Material=@n16.MA_Out, "Customization Mask (_CM B)"=@n17) {materialFunction=/Game/Weapons/WeaponSkins/Wraps/MaterialFunctions/MF_Effect_Hue_Shift}.Result
    "Customization Mask (_CM B)"=@n17
   ) {materialFunction=/Game/Weapons/WeaponSkins/Wraps/MaterialFunctions/MF_Effect_Audio_Pulse}.Result)
  FunctionOutput(None=NamedRerouteDeclaration(None=@n4.Decal_Mask_RGBA) {name=Decal_Mask_RGBA, nodeColor=<rgba 0 1 1 1>}) {outputName=Decal_Mask_RGBA}
  FunctionOutput(None=NamedRerouteDeclaration(None=@n4.Decal_Mask_Alpha) {name=Decal_Mask_A, nodeColor=<rgba 0 1 1 1>}) {outputName=Decal_Mask_Alpha}
  FunctionOutput(None=@n8.Hotwax_Msk) {outputName=Hotwax_Msk}
  FunctionOutput(None=NamedRerouteDeclaration(None=@n3.Decalmulti_Msk) {name=Decal_Multi_Msk, nodeColor=<rgba 0 0.0625 1 1>}) {outputName=Decal_Multi_Msk}
  MaterialFunctionCall!(Material=NamedRerouteDeclaration(None=FunctionInput() {inputName=In_Material, inputType=FunctionInput_MaterialAttributes, sortPriority=-200}) {name=In_Material, nodeColor=<rgba 1 0.6328 0 1>}) {materialFunction=/Engine/Functions/MaterialLayerFunctions/MatLayerBlend_BreakNormal}
  MaterialFunctionCall!(
   Material=@n7.Result
   Mask Crevise=NamedRerouteUsage()
   Mask Primary=NamedRerouteUsage()
   Mask Secondary=NamedRerouteUsage()
   TS_Mask_02=NamedRerouteUsage()
   TS_Mask_01=NamedRerouteUsage()
   Left_Right_Mask
  ) {materialFunction=/Game/Weapons/WeaponSkins/Wraps/MaterialFunctions/MF_Effect_Metal_LUT}
  MaterialFunctionCall!(Material=@n7.Result, Customization Mask A=NamedRerouteUsage(), Curavture=NamedRerouteUsage(), Customization Mask B=NamedRerouteUsage()) {materialFunction=/Game/Weapons/WeaponSkins/Wraps/MaterialFunctions/MF_Effect_Toon}
  n18 = NamedRerouteUsage()
  MaterialFunctionCall!(
   Material=@n3.Result
   Decal_Mask
   Mask Crevise=NamedRerouteUsage()
   "Customization Mask (_CM B)"=Max(A=NamedRerouteUsage(), B=@n18)
   Mask Primary=NamedRerouteUsage()
   Mask Secondary=@n18
   TS_Mask_01=NamedRerouteUsage()
   TS_Mask_03
   TS_Mask_04
   TS_Mask_02=NamedRerouteUsage()
  ) {materialFunction=/Game/Weapons/WeaponSkins/Wraps/MaterialFunctions/MF_Effect_Iridescent}
  NamedRerouteDeclaration(None=FunctionInput() {bUsePreviewValueAsDefault=true, inputName=Object-SafeZone-Back, inputType=FunctionInput_Vector4, sortPriority=100}) {name=SafeZone-Back, nodeColor=<rgba 1 0.6588 0.0581 1>}
  NamedRerouteDeclaration(None=FunctionInput() {bUsePreviewValueAsDefault=true, inputName=Object-SafeZone-Middle, inputType=FunctionInput_Vector4, sortPriority=101}) {name=SafeZone-Middle, nodeColor=<rgba 1 0.6588 0.0581 1>}
  n19 = MaterialFunctionCall(Float4=NamedRerouteDeclaration(None=FunctionInput() {inputName=IsState_B_V4, inputType=FunctionInput_Vector4, sortPriority=-1}) {name=Is_StatsB_V4, nodeColor=<rgba 0 0.9253 1 1>}) {materialFunction=&BreakOutFloat4Components}
  NamedRerouteDeclaration(None=@n19.G) {name=IsPickaxe, nodeColor=<rgba 0 0.9253 1 1>}
  NamedRerouteDeclaration(None=FunctionInput() {inputName=IsState_B, inputType=FunctionInput_Vector2, sortPriority=-1}) {name=Is_StatesB, desc="R= Backpack  B=Pickaxe", nodeColor=<rgba 0 0.9253 1 1>}
  NamedRerouteDeclaration(None=FunctionInput() {inputName=Ts_Mask_01, inputType=FunctionInput_Vector4, sortPriority=20}) {name=TS_Mask_01, nodeColor=<rgba 1 0 0.7031 1>}
  NamedRerouteDeclaration(None=FunctionInput() {inputName=Ts_Mask_02, inputType=FunctionInput_Vector4, sortPriority=20}) {name=TS_Mask_02, nodeColor=<rgba 1 0 0.7031 1>}
  NamedRerouteDeclaration(None=FunctionInput() {bUsePreviewValueAsDefault=true, inputName=Pre-Skinned Local Position, sortPriority=41}) {name=PreSkin_Position, nodeColor=<rgba 0.6039 0.1176 0.4118 1>}
  NamedRerouteDeclaration(None=FunctionInput() {bUsePreviewValueAsDefault=true, inputName=Pre-Skinned Local Normal, sortPriority=42}) {name=PreSkin_Normal, nodeColor=<rgba 0.3451 0.3098 1 1>}
  NamedRerouteDeclaration(None=FunctionInput() {inputName=Root_Material, inputType=FunctionInput_MaterialAttributes, sortPriority=-200}) {name=Root_Material_1, nodeColor=<rgba 1 0.6328 0 1>}
  NamedRerouteDeclaration(None=FunctionInput() {inputName=Caculated_Bounds}) {name=Caculated_Bounds, nodeColor=<rgba 0.465 1 0.0095 1>}
  n20 = MaterialFunctionCall(Float4=NamedRerouteDeclaration(None=FunctionInput() {bUsePreviewValueAsDefault=true, inputName="Customization Mask (_CM RGBA)", inputType=FunctionInput_Vector4, sortPriority=26}) {name=Masks_RGBA, nodeColor=<rgba 0.0094 0.0094 0.901 1>}) {materialFunction=&BreakOutFloat4Components}
  NamedRerouteDeclaration(None=@n20.R) {name=Mask_Primary, nodeColor=<rgba 0.0094 0.0094 0.901 1>}
  NamedRerouteDeclaration(None=@n20.G) {name=Mask_Crevice, nodeColor=<rgba 0 0.7865 0 1>}
  NamedRerouteDeclaration(None=FunctionInput() {bUsePreviewValueAsDefault=true, inputName=Object-SafeZone-Front, inputType=FunctionInput_Vector4, sortPriority=102}) {name=SafeZone-Front, nodeColor=<rgba 1 0.6588 0.0581 1>}
  NamedRerouteDeclaration(None=@n20.B) {name=Mask_Secondary, nodeColor=<rgba 0.0094 0.0094 0.901 1>}
  NamedRerouteDeclaration() {name=Bounds_Max, nodeColor=<rgba 1 0 0.4811 1>}
  NamedRerouteDeclaration() {name=Bounds_Min, nodeColor=<rgba 1 0 0.4811 1>}
  n21 = FunctionInput() {bUsePreviewValueAsDefault=true, inputName=Object-SafeZone-VP_Charcter_Preview_Switch, inputType=FunctionInput_Scalar, sortPriority=105}
  NamedRerouteDeclaration(None=@n21) {name=SafeZone-VP_Charcter_Preview_Switch, nodeColor=<rgba 1 0.6588 0.0581 1>}
  NamedRerouteDeclaration(None=@n20.A) {name=Mask_Gmap, nodeColor=<rgba 0.0094 0.0094 0.901 1>}
  NamedRerouteDeclaration(None=FunctionInput() {bUsePreviewValueAsDefault=true, inputName=Forward Axis}) {name=Forward Axis, nodeColor=<rgba 0.7812 0 1 1>}
  NamedRerouteDeclaration(None=@n19.A) {name=isCharAccent, nodeColor=<rgba 0 0.9253 1 1>}
  n22 = FunctionInput() {bUsePreviewValueAsDefault=true, inputName=Object-SafeZone-Virtual-Pivot, inputType=FunctionInput_Vector4, sortPriority=105}
  NamedRerouteDeclaration(None=@n22) {name=SafeZone-Virtual-Pivot, nodeColor=<rgba 1 0.6588 0.0581 1>}
  NamedRerouteDeclaration(None=FunctionInput() {inputName=Ts_Emis_01_Msk, inputType=FunctionInput_Scalar, sortPriority=20}) {name=TS_Emis_01_Msk, nodeColor=<rgba 1 0.8672 0 1>}
  NamedRerouteDeclaration(None=FunctionInput() {inputName=Ts_Emis_02_Msk, inputType=FunctionInput_Scalar, sortPriority=20}) {name=TS_Emis_02_Msk, nodeColor=<rgba 1 0.8672 0 1>}
  NamedRerouteDeclaration() {name=Bounds_Center, nodeColor=<rgba 1 0 0.4811 1>}
  NamedRerouteDeclaration(None=FunctionInput() {inputName="Msk_Ts01<--->Ts02", inputType=FunctionInput_Scalar}) {name="MSk_TS1<->Ts2", nodeColor=<rgba 0 0.9531 1 1>}
  NamedRerouteDeclaration() {name=Scale_Multiplier, nodeColor=<rgba 1 0 0.3516 1>}
  n23 = MaterialFunctionCall(Float4=NamedRerouteDeclaration(None=FunctionInput() {inputName=IsStateS, inputType=FunctionInput_Vector4, sortPriority=-1}) {name=Is_States, desc="R=vehicle G=VehicleSide =R=Mech A=Character", nodeColor=<rgba 0 0.9253 1 1>}) {materialFunction=&BreakOutFloat4Components}
  NamedRerouteDeclaration(None=@n23.B) {name=isMech, nodeColor=<rgba 0 0.9253 1 1>}
  NamedRerouteDeclaration(None=StaticBoolParameter() {group=Customization, parameterName="Custmization  V2", sortPriority=-500}) {name="Custmization  V2 State", nodeColor=<rgba 1 0.1172 0 1>}
  NamedRerouteDeclaration(None=@n23.G) {name=isVehicleSide, nodeColor=<rgba 0 0.9253 1 1>}
  NamedRerouteDeclaration(None=@n23.R) {name=isVehicle, nodeColor=<rgba 0 0.9253 1 1>}
  NamedRerouteDeclaration(None=@n23.A) {name=isCharacter, nodeColor=<rgba 0 0.9253 1 1>}
  NamedRerouteDeclaration(None=@n19.R) {name=IsBackpack, nodeColor=<rgba 0 0.9253 1 1>}
  Reroute()
 ]}
