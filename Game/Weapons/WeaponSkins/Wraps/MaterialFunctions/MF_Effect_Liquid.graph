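# Material function for weapon-skin wraps that adds an animated liquid-behind-glass effect to the input material, gated by the Effect_Liquid static switch.
# Bounds-based projection places the fill level, with offsets, scales and glass rotations blended per object type through IsState inputs.
# Layered sine waves distort the surface, with strength eased over time from a reactive timestamp.
# Curve-atlas ramps colour the emissive liquid and output a mask, while CM masks drive material blending.
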MaterialFunction /Game/Weapons/WeaponSkins/Wraps/MaterialFunctions/MF_Effect_Liquid
graph {nodes [
  n1 = ScalarParameter() {group=Effect_Liquid_Liq, parameterName=EFLiq_Surface_A values, sortPriority=0}
  n2 = Multiply(A=NamedRerouteUsage(), B=NamedRerouteUsage())
  n3 = Multiply(
   A=Add(B=@n2) {constA=0.2131}
   B=ScalarParameter() {defaultValue=2, group=Effect_Liquid_Liq, parameterName=EFLiq_Surface_Distortion_TileRate A Master}
  )
  n4 = ScalarParameter() {defaultValue=4.3, group=Effect_Liquid_Liq, parameterName=EFLiq_Surface_Distortion_TileRate A-A}
  n5 = Fmod(
   A=MaterialFunctionCall(Speed=ScalarParameter() {defaultValue=0.5, group=Effect_Liquid_Liq, parameterName=EFLiq_Surface_Distortion_Speed Max}) {materialFunction=&TimeWithSpeedVariable}.Time
   B=Constant() {r=1000}
  )
  n6 = Sine(None=Add(A=Multiply(A=@n3, B=@n4) {constB=14.4956}, B=@n5))
  n7 = Multiply(A=@n3, B=ScalarParameter() {defaultValue=2.13, group=Effect_Liquid_Liq, parameterName=EFLiq_Surface_Distortion_TileRate A-B}) {constB=-5.8402}
  n8 = Multiply(A=Sine(None=Add(A=@n7, B=@n5))) {constB=0.5}
  n9 = Sine(None=Add(A=@n3, B=@n5))
  n10 = MaterialFunctionCall(Float2=Add(A=Multiply(A=@n6) {constB=0.25}, B=Add(A=@n8, B=@n9))) {materialFunction=&BreakOutFloat2Components}
  n11 = VectorParameter() {group=Effect_Liquid_Liq | Reactive, parameterName=EFLiq_Surface_Distortion_Vector Strenght, defaultValue=<rgba 1 1 1 0>}
  n12 = MaterialFunctionCall(Float2=Multiply(A=ScalarParameter() {defaultValue=0.25, group=Effect_Liquid_Liq, parameterName=EFLiq_Surface_Distortion_Strength_Rest}, B=@n11.RGB)) {materialFunction=&BreakOutFloat2Components}
  n13 = Add(A=Multiply(A=@n10.R, B=@n12.R), B=Multiply(A=@n10.G, B=@n12.G))
  n14 = MaterialFunctionCall() {materialFunction=&ObjectScale}
  n15 = Multiply(A=@n2, B=ScalarParameter() {defaultValue=2, group=Effect_Liquid_Liq, parameterName=EFLiq_Surface_Distortion_TileRate B Master})
  n16 = ScalarParameter() {defaultValue=1.273, group=Effect_Liquid_Liq, parameterName=EFLiq_Surface_Distortion_TileRate B-A}
  n17 = ScalarParameter() {defaultValue=-1.5, group=Effect_Liquid_Liq, parameterName=EFLiq_Surface_Distortion_Speed B Max}
  n18 = MaterialFunctionCall(Speed=@n17) {materialFunction=&TimeWithSpeedVariable}
  n19 = Constant() {r=10000.0312}
  n20 = Sine(None=Add(A=Multiply(A=@n15, B=@n16) {constB=14.4956}, B=Fmod(A=@n18.Time, B=@n19)))
  n21 = Multiply(A=@n15, B=ScalarParameter() {defaultValue=3.0133, group=Effect_Liquid_Liq, parameterName=EFLiq_Surface_Distortion_TileRate B-B}) {constB=-5.8402}
  n22 = ScalarParameter() {defaultValue=1, group=Effect_Liquid_Liq, parameterName=EFLiq_Surface_Distortion_Speed C Mult}
  n23 = Fmod(
   A=MaterialFunctionCall(Speed=Multiply(A=@n17, B=@n22)) {materialFunction=&TimeWithSpeedVariable}.Time
   B=@n19
  )
  n24 = Multiply(A=Sine(None=Add(A=@n21, B=@n23))) {constB=0.5}
  n25 = MaterialFunctionCall(Speed=Multiply(A=@n17, B=ScalarParameter() {defaultValue=1, group=Effect_Liquid_Liq, parameterName=EFLiq_Surface_Distortion_Speed D Mult})) {materialFunction=&TimeWithSpeedVariable}
  n26 = Sine(None=Add(A=@n15, B=Fmod(A=@n25.Time, B=@n19)))
  n27 = MaterialFunctionCall(Float2=Add(A=Multiply(A=@n20) {constB=0.25}, B=Add(A=@n24, B=@n26))) {materialFunction=&BreakOutFloat2Components}
  n28 = MaterialFunctionCall(Float2=@n11.RGB) {materialFunction=&BreakOutFloat2Components}
  n29 = Multiply(A=@n27.R, B=@n28.R)
  n30 = Multiply(A=@n27.G, B=@n28.G)
  n31 = Divide(
   A=Subtract(
    A=Time()
    B=ScalarParameter() {defaultValue=-1000000000, group=Effect_Liquid_Liq | Reactive, parameterName=EFLiq_TimeFrom-BP, sortPriority=100}
   )
   B=ScalarParameter() {defaultValue=1, group=Effect_Liquid_Liq | Reactive, parameterName=EFLiq_Time Druation, sortPriority=100}
  )
  n32 = Multiply(
   A=Multiply(A=Add(A=@n29, B=@n30), B=@n14.Scale Z) {constB=0.75}
   B=LinearInterpolate(
    A=ScalarParameter() {defaultValue=0.025, group=Effect_Liquid_Liq | Reactive, parameterName=EFLiq_Surface_Distortion_StrenghtMX}
    B=ScalarParameter() {defaultValue=0.005, group=Effect_Liquid_Liq | Reactive, parameterName=EFLiq_Surface_Distortion_Strenght}
    Alpha=Saturate(None=@n31)
   )
  )
  Add(A=Multiply(A=Multiply(A=@n1, B=@n13), B=@n14.Scale Z), B=@n32)
  Constant4Vector() {constant=<rgba 0 0 5 -1>}
  n33 = TextureSample() {samplerType=SAMPLERTYPE_Masks, texture=/Game/Weapons/FORT_Rifles/Textures/T_Scar_CM}
  n34 = MaterialFunctionCall(Float4=FunctionInput(Preview=@n33.RGBA) {inputName=CM_Msk, inputType=FunctionInput_Vector4, sortPriority=4}) {materialFunction=&BreakOutFloat4Components}
  n35 = NamedRerouteDeclaration(None=@n34.R) {name=Msk_Prime, nodeColor=<rgba 0 1 0.7422 1>}
  n36 = NamedRerouteDeclaration(None=@n34.B) {name=Msk_Accent, nodeColor=<rgba 0 1 0.7656 1>}
  n37 = BlendMaterialAttributes(A=NamedRerouteUsage(), B=NamedRerouteUsage(), Alpha=Max(B=Max(A=@n35, B=@n36))) {vertexAttributeBlendType=UseA}
  n38 = TextureSample() {samplerType=SAMPLERTYPE_Normal, texture=/Game/Weapons/FORT_Rifles/T_SCAR_N}
  n39 = Constant3Vector() {constant=<rgba 1 0 0 0>}
  n40 = GetMaterialAttributes(None=FunctionInput(Preview=SetMaterialAttributes!(MaterialAttributes, Normal=@n38.RGB, BaseColor=@n39) {attributeSetTypes [#1, #2]}) {inputName=Material, inputType=FunctionInput_MaterialAttributes}) {attributeGetTypes [
    #2
    #3
    #1
    #4
   ]}
  FunctionOutput(None=StaticSwitch(
    True=BlendMaterialAttributes(A=@n37, Alpha=@n36) {vertexAttributeBlendType=UseA}
    False=NamedRerouteDeclaration(None=@n40.MaterialAttributes) {name=Root_Material, nodeColor=<rgba 0.7812 1 0 1>}
    Value=StaticBoolParameter() {group=Customization, parameterName=Effect_Liquid}
   ))
  n41 = Add(A=NamedRerouteUsage()) {constB=-0.5}
  n42 = ScalarParameter() {defaultValue=0.025, group=Effect_Liquid, parameterName=EFLiq_CC_Surface_Warp}
  n43 = Add(A=Add(A=Multiply(A=@n41, B=@n42), B=NamedRerouteUsage()), B=NamedRerouteUsage()) {constB=0}
  n44 = CurveAtlasRowParameter(CurveTime=@n43) {
   atlas=/Game/Weapons/WeaponSkins/Wraps/Textures/CruveAtlas_Wraps_01
   curve=/Game/Weapons/WeaponSkins/Wraps/Textures/0_ColorCurves/CC_BlueLiq
   defaultValue=164
   group=Effect_Liquid
   parameterName=EFLiq_CC_Masks
  }
  FunctionOutput(None=MaterialFunctionCall!(
    Index=ScalarParameter() {defaultValue=164, group=Effect_Liquid, parameterName="EFLiq_CC_Masks  index"}
    CCinput=@n44
    CCinput_Alpha=@n44
    UV=@n43
   ) {materialFunction=&/MF_Global_CC_Tex}.4v) {outputName=Effect Liquid Msk}
  n45 = NamedRerouteUsage()
  n46 = VectorParameter() {group=Effect_Liquid_UV, parameterName=EFLiq_Pivot_PosB, defaultValue=<rgba -0.5 -0.5 -0.5 1>}
  n47 = NamedRerouteUsage()
  n48 = Transform(None=Add(A=Add(A=@n45, B=@n46.RGB), B=@n47) {constB=0}) {transformSourceType=TRANSFORMSOURCE_Local}
  n49 = VectorParameter() {group=Effect_Liquid_UV, parameterName=EFLiq_Pivot_PosVeh, defaultValue=<rgba 0 0.5 0 1>}
  n50 = MaterialFunctionCall(Float4=NamedRerouteUsage()) {materialFunction=&BreakOutFloat4Components}
  n51 = ScalarParameter() {group=Effect_Liquid_UV, parameterName="EFLiq_Liquid Veh 0=World value", sortPriority=0}
  n52 = LinearInterpolate(A=Add(A=Add(A=@n45, B=@n49.RGB), B=@n47) {constB=0}, B=@n48, Alpha=Max(A=@n50.G, B=@n51))
  n53 = NamedRerouteUsage()
  n54 = Constant()
  n55 = MaterialFunctionCall(R=@n54, G=@n51, B=@n54, A=@n54) {materialFunction=&AppendMany}
  n56 = MaterialFunctionCall!(
   Normal_Map
   Offset Rt
   Pivot
   Caculated_Bounds=LinearInterpolate(A=LinearInterpolate(A=@n48, B=@n52, Alpha=@n50.R), B=@n48, Alpha=@n50.B)
   Pre_Skinned_Normals=NamedRerouteUsage()
   is_State=Saturate(None=Add(A=@n53, B=@n55.RGBA))
   UV_Rotation=ScalarParameter() {desc="use more for roll. ", group=Effect_Liquid_UV, parameterName=EFLiq_Surface_Z_Rotation, sortPriority=0}
   Uvs_Offset
   Uvs_TileRate
   Uvs_Offset_Pivot
   For_Safe_Zones
   Rotation_Comp
   Clamp Y
   Clamp X
   Clamp Padding_Comp
   No_Rotation
  ) {materialFunction=&/MF_Projection_Side_Normal_V2}
  n57 = Power(Base=ComponentMask(None=Add(A=@n56.UVs) {constB=-0.5}) {g=true})
  NamedRerouteDeclaration(None=Multiply(A=Saturate(None=@n57), B=@n32)) {name=Surface_Distortion, nodeColor=<rgba 0.4531 0 1 1>}
  NamedRerouteDeclaration(None=@n34.G) {name=Msk_Curvature, nodeColor=<rgba 0 1 0.7422 1>}
  n58 = FunctionInput(Preview=Constant4Vector() {constant=<rgba 1.9782 25.3676 9.2814 -0.756>}.RGBA) {bUsePreviewValueAsDefault=true, inputName=Object-SafeZone-Virtual-Pivot, inputType=FunctionInput_Vector4, sortPriority=105}
  NamedRerouteDeclaration(None=@n58) {name=SafeZone VP, nodeColor=<rgba 1 0 0.6562 1>}
  n59 = VectorParameter() {group=Effect_Liquid | Glass | UVs, parameterName=EFLiq_Glass_Rotation_Weapon, defaultValue=<rgba 0 0 -0.25 1>}
  n60 = VectorParameter() {group=Effect_Liquid | Glass | UVs, parameterName=EFLiq_Glass_Rotation_Vehicle, defaultValue=<rgba 0 0 0 1>}
  n61 = VectorParameter() {group=Effect_Liquid | Glass | UVs, parameterName=EFLiq_Glass_Rotation_Vehicle_Side, defaultValue=<rgba 0 0 0 1>}
  n62 = MaterialFunctionCall(Float4=NamedRerouteUsage()) {materialFunction=&BreakOutFloat4Components}
  n63 = VectorParameter() {group=Effect_Liquid | Glass | UVs, parameterName=EFLiq_Glass_Rotation_Mech, defaultValue=<rgba 0.25 0.25 0 1>}
  n64 = VectorParameter() {group=Effect_Liquid | Glass | UVs, parameterName=EFLiq_Glass_Rotation_Character, defaultValue=<rgba 0.25 0.25 0 1>}
  n65 = LinearInterpolate(
   A=LinearInterpolate(A=@n59.RGB, B=LinearInterpolate(A=@n60.RGB, B=@n61.RGB, Alpha=@n62.G), Alpha=@n62.R)
   B=LinearInterpolate(A=@n63.RGB, B=@n64.RGB, Alpha=@n62.A)
   Alpha=@n62.B
  )
  n66 = VectorParameter() {group=Effect_Liquid | Glass | UVs, parameterName=EFLiq_Glass_Rotation_Backpack, defaultValue=<rgba 0 0 0 1>}
  n67 = MaterialFunctionCall(Float2=NamedRerouteUsage()) {materialFunction=&BreakOutFloat2Components}
  n68 = VectorParameter() {group=Effect_Liquid | Glass | UVs, parameterName=EFLiq_Glass_Rotation_Pickaxe, defaultValue=<rgba 0 0 0 1>}
  NamedRerouteDeclaration(None=LinearInterpolate(A=LinearInterpolate(A=@n65, B=@n66.RGB, Alpha=@n67.R), B=@n68.RGB, Alpha=@n67.G)) {name=Glass Rotation, nodeColor=<rgba 0 0.0156 1 1>}
  n69 = FunctionInput(Preview=Constant()) {bUsePreviewValueAsDefault=true, inputName=Object-SafeZone-VP-Character-PreView-Switch, inputType=FunctionInput_Scalar, sortPriority=105}
  NamedRerouteDeclaration(None=@n69) {name=SafeZone Character-Switch, nodeColor=<rgba 1 0.1875 0 1>}
  NamedRerouteDeclaration(None=MaterialFunctionCall(
    Caculated_Bounds=NamedRerouteUsage()
    Pre_Skinned_Normals=NamedRerouteUsage()
    is_State=NamedRerouteUsage()
    Uvs_TileRate=Constant()
    For_Safe_Zones=StaticBool() {value=true}
   ) {materialFunction=&MF_Projection_Side_Normal_V2}.UVs No Offsets/Tiling) {name=UVs_offset_1, nodeColor=<rgba 0.1719 1 0 1>}
  n70 = ScalarParameter() {defaultValue=1, group=Effect_Liquid | Glass | UVs, parameterName=EFLiq_Glass_Scale_Vehicle, sortPriority=150}
  n71 = ScalarParameter() {defaultValue=1, group=Effect_Liquid | Glass | UVs, parameterName=EFLiq_Glass_Scale_Vehicle_Side, sortPriority=150}
  n72 = MaterialFunctionCall(Float4=NamedRerouteUsage()) {materialFunction=&BreakOutFloat4Components}
  n73 = LinearInterpolate(
   A=LinearInterpolate(
    A=ScalarParameter() {defaultValue=1, group=Effect_Liquid | Glass | UVs, parameterName=EFLiq_Glass_Scale_Weapon, sortPriority=150}
    B=LinearInterpolate(A=@n70, B=@n71, Alpha=@n72.G)
    Alpha=@n72.R
   )
   B=LinearInterpolate(
    A=ScalarParameter() {defaultValue=1, group=Effect_Liquid | Glass | UVs, parameterName=EFLiq_Glass_Scale_Mech, sortPriority=150}
    B=ScalarParameter() {defaultValue=1, group=Effect_Liquid | Glass | UVs, parameterName=EFLiq_Glass_Scale_Character, sortPriority=150}
    Alpha=@n72.A
   )
   Alpha=@n72.B
  )
  n74 = ScalarParameter() {defaultValue=1, group=Effect_Liquid | Glass | UVs, parameterName=EFLiq_Glass_Scale_Backpack, sortPriority=150}
  n75 = MaterialFunctionCall(Float2=NamedRerouteUsage()) {materialFunction=&BreakOutFloat2Components}
  NamedRerouteDeclaration(None=LinearInterpolate(
    A=LinearInterpolate(A=@n73, B=@n74, Alpha=@n75.R)
    B=ScalarParameter() {defaultValue=1, group=Effect_Liquid | Glass | UVs, parameterName=EFLiq_Glass_Scale_Pickaxe, sortPriority=150}
    Alpha=@n75.G
   )) {name=Glass_Scale, nodeColor=<rgba 0 0.7188 1 1>}
  n76 = TextureObjectParameter() {
   group=Effect_Liquid | Glass
   parameterName=EFLiq_Glass_BaseColor
   samplerType=SAMPLERTYPE_Masks
   texture=&T_BlackMask
  }
  n77 = ScalarParameter() {group=Effect_Liquid | Glass, parameterName=EFLiq_Glass_Alt_Axis Contrast}
  n78 = NamedRerouteUsage()
  n79 = NamedRerouteUsage()
  n80 = ScalarParameter() {group=Effect_Liquid | Glass, parameterName=EFLiq_Glass_Alt_Axis}
  n81 = NamedRerouteUsage()
  n82 = NamedRerouteUsage()
  NamedRerouteDeclaration(None=MaterialFunctionCall!(
    Texture=@n76
    Caculated_Bounds=NamedRerouteUsage()
    Pre_Skinned_Normals=NamedRerouteUsage()
    Blend_Contrast=@n77
    IsStatsB=@n78
    IsState=@n79
    Fill_alternate_Axis_Opacity=@n80
    "Fill_alternate_Axis_Color_A=Roughness"=VectorParameter() {group=Effect_Liquid | Glass, parameterName="EFLiq_Glass_Alt_Axis Color   A=Roughness"}.RGBA
    Offset
    Scale=@n81
    Offset_Pivot
    Rotation_Axis=@n82
   ) {materialFunction=&/MF_TopsSideProjection_Base_V2}.Combined) {name=Glass BCA, nodeColor=<rgba 1 0 0.6562 1>}
  n83 = TextureObjectParameter() {
   group=Effect_Liquid | Glass
   parameterName=EFLiq_Glass_Normals
   samplerType=SAMPLERTYPE_Normal
   texture=&FlatNormalBlue
  }
  NamedRerouteDeclaration(None=MaterialFunctionCall!(
    Texture_Normals=@n83
    Rotation_Axis=@n82
    Caculated_Bounds=NamedRerouteUsage()
    Pre_Skinned_Normals=NamedRerouteUsage()
    IsStatsB=@n78
    IsState=@n79
    Blend_Contrast=@n77
    Fill_alternate_Axis_Opacity=@n80
    Scale=@n81
    Offset_Pivot
    Offset
   ) {materialFunction=&/MF_TopsSideProjection_Normals_V2}.Add_Surface_Normal) {name=Glass Normal, nodeColor=<rgba 0 0.9766 1 1>}
  n84 = NamedRerouteUsage()
  n85 = VectorParameter() {group=Effect_Liquid, parameterName=EFLiq__Gmap_Color, defaultValue=<rgba 0 0 0 1>}
  n86 = VectorParameter() {group=Effect_Liquid, parameterName=EFLiq__Gmap--, channelNames {a="", b=Max, g=Min, r="0=Use Gmap"}, defaultValue=<rgba 1 0 1 1>}
  n87 = Max(A=SmoothStep(Min=@n86.Min, Max=@n86.Min, Value=NamedRerouteUsage()), B=@n86."0=Use Gmap")
  n88 = ScalarParameter() {defaultValue=164, group=Effect_Liquid, parameterName="EFLiq_CC_Surface  index"}
  n89 = CurveAtlasRowParameter(CurveTime=@n43) {
   atlas=&CruveAtlas_Wraps_01
   curve=&CC_BlueLiq
   defaultValue=164
   group=Effect_Liquid
   parameterName=EFLiq_CC_Surface
  }
  n90 = MaterialFunctionCall(Index=@n88, CCinput_Alpha=@n89.A, UV=@n43) {materialFunction=&MF_Global_CC_Tex}
  n91 = Multiply(A=@n43, B=@n90.Alpha)
  n92 = CurveAtlasRowParameter(CurveTime=@n91) {
   atlas=&CruveAtlas_Wraps_01
   curve=&CC_BlueLiq
   defaultValue=164
   group=Effect_Liquid
   parameterName=EFLiq_CC_Surface
  }
  n93 = Multiply(
   A=MaterialFunctionCall(Index=@n88, CCinput=@n92, UV=@n91) {materialFunction=&MF_Global_CC_Tex}.3V
   B=ScalarParameter() {defaultValue=1, group=Effect_Liquid, parameterName=EFLiq_CC_Surface_Brightness}
  )
  n94 = Add(
   A=VectorParameter() {group=Effect_Liquid, parameterName=EFLiq__GlassColor, defaultValue=<rgba 0 0 0 1>}.RGB
   B=Multiply(A=LinearInterpolate(A=@n85, Alpha=@n87), B=LinearInterpolate(B=@n93, Alpha=@n90.Alpha))
  )
  n95 = VectorParameter() {group=Effect_Liquid, parameterName=EFLiq_Curvature_Color, defaultValue=<rgba 1 1 1 1>}
  n96 = Multiply(A=@n41) {constB=2}
  n97 = SetMaterialAttributes!(
   MaterialAttributes=NamedRerouteUsage()
   BaseColor=StaticSwitchParameter(
    True=ComponentMask(None=@n84) {b=true, g=true, r=true}
    False=VectorParameter() {group=Effect_Liquid, parameterName=EFLiq__BaseColor, defaultValue=<rgba 0 0 0 1>}.RGB
   ) {group=Effect_Liquid | Glass, parameterName=EFLiq_Glass---BC_Texture, sortPriority=0}
   Roughness=StaticSwitchParameter(True=ComponentMask(None=@n84) {a=true}, False=ScalarParameter() {defaultValue=0.5, group=Effect_Liquid, parameterName=EFLiq__Roughness}) {group=Effect_Liquid | Glass, parameterName=EFLiq_Glass---BC_Texture, sortPriority=0}
   Specular=ScalarParameter() {defaultValue=0.5, group=Effect_Liquid, parameterName=EFLiq__Spec}
   Metallic=ScalarParameter() {group=Effect_Liquid, parameterName=EFLiq__Metallic}
   Normal=StaticSwitchParameter(True=NamedRerouteUsage(), False=NamedRerouteUsage()) {group=Effect_Liquid | Glass, parameterName=EFLiq_Glass---Normal_Texture, sortPriority=0}
   EmissiveColor=LinearInterpolate(A=@n94, B=Multiply(A=@n94, B=@n95.RGB), Alpha=Clamp(None=@n96))
  ) {attributeSetTypes [
    #2
    #3
    #5
    #4
    #1
    #6
   ]}
  NamedRerouteDeclaration(None=@n97) {name=Mat_Base, nodeColor=<rgba 0 0.3906 1 1>}
  n98 = VectorParameter() {group=Effect_Liquid_UV, parameterName=EFLiq_End_Offset_Weapon}
  n99 = MaterialFunctionCall(Float4=NamedRerouteUsage()) {materialFunction=&BreakOutFloat4Components}
  n100 = LinearInterpolate(
   A=VectorParameter() {group=Effect_Liquid_UV, parameterName=EFLiq_End_Offset_Vehicle, defaultValue=<rgba 0 0 0 1>}.RGB
   B=VectorParameter() {group=Effect_Liquid_UV, parameterName=EFLiq_End_Offset_VehicleSide, defaultValue=<rgba 0 0 0 1>}.RGB
   Alpha=@n99.G
  )
  n101 = VectorParameter() {group=Effect_Liquid_UV, parameterName=EFLiq_End_Offset_Mech, defaultValue=<rgba 0 0 0 1>}
  n102 = VectorParameter() {group=Effect_Liquid_UV, parameterName=EFLiq_End_Offset_Character, defaultValue=<rgba 0 0 0 1>}
  n103 = MaterialFunctionCall(Float2=NamedRerouteUsage()) {materialFunction=&BreakOutFloat2Components}
  n104 = LinearInterpolate(
   A=LinearInterpolate(A=LinearInterpolate(A=@n98.RGB, B=@n100, Alpha=@n99.R), B=LinearInterpolate(A=@n101.RGB, B=@n102.RGB, Alpha=@n99.A), Alpha=@n99.B)
   B=VectorParameter() {group=Effect_Liquid_UV, parameterName=EFLiq_End_Offset_Backpack, defaultValue=<rgba 0 0 0 1>}.RGB
   Alpha=@n103.R
  )
  n105 = VectorParameter() {group=Effect_Liquid_UV, parameterName=EFLiq_End_Offset_Pickaxe, defaultValue=<rgba 0 0 0 1>}
  n106 = VectorParameter() {group=Effect_Liquid_UV, parameterName=EFLiq_Offset_Weapon, defaultValue=<rgba 0 0.21 0.012 1>}
  n107 = MaterialFunctionCall(Float4=NamedRerouteUsage()) {materialFunction=&BreakOutFloat4Components}
  n108 = LinearInterpolate(
   A=VectorParameter() {group=Effect_Liquid_UV, parameterName=EFLiq_Offset_Vehicle, defaultValue=<rgba 0 0 0 1>}.RGB
   B=VectorParameter() {group=Effect_Liquid_UV, parameterName=EFLiq_Offset_VehicleSide, defaultValue=<rgba 0 0 0 1>}.RGB
   Alpha=@n107.G
  )
  n109 = VectorParameter() {group=Effect_Liquid_UV, parameterName=EFLiq_Offset_Mech, defaultValue=<rgba 0 0 0 1>}
  n110 = VectorParameter() {group=Effect_Liquid_UV, parameterName=EFLiq_Offset_Character, defaultValue=<rgba 0 0 0 1>}
  n111 = MaterialFunctionCall(Float2=NamedRerouteUsage()) {materialFunction=&BreakOutFloat2Components}
  n112 = LinearInterpolate(
   A=LinearInterpolate(A=LinearInterpolate(A=@n106.RGB, B=@n108, Alpha=@n107.R), B=LinearInterpolate(A=@n109.RGB, B=@n110.RGB, Alpha=@n107.A), Alpha=@n107.B)
   B=VectorParameter() {group=Effect_Liquid_UV, parameterName=EFLiq_Offset_Backpack, defaultValue=<rgba 0 0 0 1>}.RGB
   Alpha=@n111.R
  )
  n113 = VectorParameter() {group=Effect_Liquid_UV, parameterName=EFLiq_Offset_Pickaxe, defaultValue=<rgba 0 0 0 1>}
  n114 = Divide(
   A=Subtract(
    A=Time()
    B=ScalarParameter() {defaultValue=-1000000000, group=Effect_Liquid_UV |Reactive, parameterName=EFLiq_Reset_TimeFrom-BP, sortPriority=100}
   )
   B=ScalarParameter() {defaultValue=1, group=Effect_Liquid_UV |Reactive, parameterName=EFLiq_Reset_Time Druation, sortPriority=100}
  )
  NamedRerouteDeclaration(None=LinearInterpolate(
    A=LinearInterpolate(A=@n104, B=@n105.RGB, Alpha=@n103.G)
    B=LinearInterpolate(A=@n112, B=@n113.RGB, Alpha=@n111.G)
    Alpha=Saturate(None=@n114)
   )) {name=UVs_Offset, nodeColor=<rgba 0 0.625 1 1>}
  NamedRerouteDeclaration(None=@n34.A) {name=Msk_Gmap, nodeColor=<rgba 0 1 0.7656 1>}
  n115 = ScalarParameter() {defaultValue=1, group=Effect_Liquid_Liq | Scale Object, parameterName=EFLiq_Surface_Scale_Vehicle, sortPriority=150}
  n116 = ScalarParameter() {defaultValue=1, group=Effect_Liquid_Liq | Scale Object, parameterName=EFLiq_Surface_Scale_Vehicle_Side, sortPriority=150}
  n117 = MaterialFunctionCall(Float4=NamedRerouteUsage()) {materialFunction=&BreakOutFloat4Components}
  n118 = LinearInterpolate(
   A=LinearInterpolate(
    A=ScalarParameter() {defaultValue=1, group=Effect_Liquid_Liq | Scale Object, parameterName=EFLiq_Surface_Scale_Weapon, sortPriority=150}
    B=LinearInterpolate(A=@n115, B=@n116, Alpha=@n117.G)
    Alpha=@n117.R
   )
   B=LinearInterpolate(
    A=ScalarParameter() {defaultValue=1, group=Effect_Liquid_Liq | Scale Object, parameterName=EFLiq_Surface_Scale_Mech, sortPriority=150}
    B=ScalarParameter() {defaultValue=1, group=Effect_Liquid_Liq | Scale Object, parameterName=EFLiq_Surface_Scale_Character, sortPriority=150}
    Alpha=@n117.A
   )
   Alpha=@n117.B
  )
  n119 = ScalarParameter() {defaultValue=1, group=Effect_Liquid_Liq | Scale Object, parameterName=EFLiq_Surface_Scale_Backpack, sortPriority=150}
  n120 = MaterialFunctionCall(Float2=NamedRerouteUsage()) {materialFunction=&BreakOutFloat2Components}
  NamedRerouteDeclaration(None=LinearInterpolate(
    A=LinearInterpolate(A=@n118, B=@n119, Alpha=@n120.R)
    B=ScalarParameter() {defaultValue=1, group=Effect_Liquid_Liq | Scale Object, parameterName=EFLiq_Surface_Scale_Pickaxe, sortPriority=150}
    Alpha=@n120.G
   )) {name=surface sacle_Scale_1, nodeColor=<rgba 0 0.7188 1 1>}
  n121 = ScalarParameter() {defaultValue=0.6, group=Effect_Liquid_par, parameterName=EFLiq_Surface_Z_Falloff_Vehicle, sortPriority=150}
  n122 = ScalarParameter() {defaultValue=0.6, group=Effect_Liquid_par, parameterName=EFLiq_Surface_Z_Falloff_Vehicle_Side, sortPriority=150}
  n123 = MaterialFunctionCall(Float4=NamedRerouteUsage()) {materialFunction=&BreakOutFloat4Components}
  n124 = LinearInterpolate(
   A=LinearInterpolate(
    A=ScalarParameter() {defaultValue=0.6, group=Effect_Liquid_par, parameterName=EFLiq_Surface_Z_Falloff_Weapon, sortPriority=150}
    B=LinearInterpolate(A=@n121, B=@n122, Alpha=@n123.G)
    Alpha=@n123.R
   )
   B=LinearInterpolate(
    A=ScalarParameter() {defaultValue=0.6, group=Effect_Liquid_par, parameterName=EFLiq_Surface_Z_Falloff_Mech, sortPriority=150}
    B=ScalarParameter() {defaultValue=0.6, group=Effect_Liquid_par, parameterName=EFLiq_Surface_Z_Falloff_Character, sortPriority=150}
    Alpha=@n123.A
   )
   Alpha=@n123.B
  )
  n125 = ScalarParameter() {defaultValue=0.6, group=Effect_Liquid_par, parameterName=EFLiq_Surface_Z_Falloff_Backpack, sortPriority=150}
  n126 = MaterialFunctionCall(Float2=NamedRerouteUsage()) {materialFunction=&BreakOutFloat2Components}
  NamedRerouteDeclaration(None=LinearInterpolate(
    A=LinearInterpolate(A=@n124, B=@n125, Alpha=@n126.R)
    B=ScalarParameter() {defaultValue=0.6, group=Effect_Liquid_par, parameterName=EFLiq_Surface_Z_Falloff_Pickaxe, sortPriority=150}
    Alpha=@n126.G
   )) {name=Liquid_Fallof, nodeColor=<rgba 0 0.7188 1 1>}
  n127 = Divide(
   A=Add(A=@n56.UVs, B=ScalarParameter() {defaultValue=-0.9, group=Effect_Liquid_par, parameterName=EFLiq_Surface_Z_Offset}) {constB=-1}
   B=NamedRerouteUsage()
  ) {constB=0.6}
  NamedRerouteDeclaration(None=Saturate(None=ComponentMask(None=@n127) {g=true})) {name=Liquid, nodeColor=<rgba 1 0.9844 0 1>}
  NamedRerouteDeclaration(None=@n33.RGBA) {name=CM_Mask, nodeColor=<rgba 1 0 0.3047 1>}
  NamedRerouteDeclaration(None=FunctionInput(Preview=Constant4Vector().RGBA) {bUsePreviewValueAsDefault=true, inputName=IsState, inputType=FunctionInput_Vector4, sortPriority=4}) {name=IsState, nodeColor=<rgba 0.1497 1 0.1633 1>}
  NamedRerouteDeclaration(None=FunctionInput(Preview=Constant2Vector()) {bUsePreviewValueAsDefault=true, inputName=IsState_B, inputType=FunctionInput_Vector2, sortPriority=4}) {name=IsState_B, nodeColor=<rgba 0.1497 1 0.1633 1>}
  n128 = PreSkinnedLocalBounds()
  n129 = Subtract(A=VertexInterpolator(VS=PreSkinnedPosition()).PS, B=@n128.Min)
  n130 = MaterialFunctionCall(Color=Subtract(A=@n128.Max, B=@n128.Max)) {materialFunction=&SplitComponents}
  n131 = Max(A=Max(A=@n130.R, B=@n130.R), B=@n130.B)
  NamedRerouteDeclaration(None=FunctionInput(Preview=Divide(A=@n129, B=@n131)) {bUsePreviewValueAsDefault=true, inputName=Calcated_Bounds, sortPriority=11}) {name=Caclcated_Bounds, nodeColor=<rgba 0 0.9297 1 1>}
  n132 = MaterialFunctionCall(Float3=Abs(None=NamedRerouteUsage())) {materialFunction=&BreakOutFloat3Components}
  n133 = Constant() {r=0.5}
  n134 = NamedRerouteUsage()
  n135 = ComponentMask(None=@n134) {b=true, desc=YZ, g=true}
  n136 = ComponentMask(None=@n134) {desc=YX, g=true, r=true}
  n137 = ComponentMask(None=@n134) {b=true, desc=XZ, r=true}
  n138 = If(A=@n132.R, B=@n133, "A > B"=@n135, "A < B"=If(A=@n132.B, B=@n133, "A > B"=@n136, "A < B"=@n137))
  n139 = Constant() {r=0.5}
  n140 = Constant() {r=0.5}
  n141 = MaterialFunctionCall(Float4=NamedRerouteUsage()) {materialFunction=&BreakOutFloat4Components}
  n142 = LinearInterpolate(
   A=If(A=@n132.B, B=@n139, "A > B"=@n136, "A < B"=If(A=@n132.R, B=@n139, "A > B"=@n135, "A < B"=@n137))
   B=If(A=@n132.G, B=@n140, "A > B"=@n137, "A < B"=If(A=@n132.B, B=@n140, "A > B"=@n136, "A < B"=@n135))
   Alpha=@n141.G
  )
  n143 = Constant() {r=0.5}
  n144 = If(A=@n132.R, B=@n143, "A > B"=@n135, "A < B"=@n136)
  NamedRerouteDeclaration(None=LinearInterpolate(A=LinearInterpolate(A=@n138, B=@n142, Alpha=@n141.R), B=If(A=@n132.G, B=@n143, "A > B"=@n137, "A < B"=@n144), Alpha=@n141.B)) {name=UVS_Cheap_Tri_1, nodeColor=<rgba 1 0 0.8438 1>}
  n145 = PreSkinnedNormal()
  NamedRerouteDeclaration(None=FunctionInput(Preview=VertexInterpolator(VS=@n145).PS) {bUsePreviewValueAsDefault=true, inputName=Pre-Skinned Local Normal, sortPriority=10}) {name=PreSkin_Normal, nodeColor=<rgba 0.8281 1 0 1>}
  NamedRerouteDeclaration(None=@n40.Normal) {name=Root_Normal, nodeColor=<rgba 0.4062 0 1 1>}
  NamedRerouteUsage()
  TransformPosition(None=WorldPosition() {worldPositionShaderOffset=WPT_ExcludeAllShaderOffsets}.XYZ) {transformSourceType=TRANSFORMPOSSOURCE_World}
  VertexNormalWS()
 ]}
