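# Primary V2 material function for base customization, layering wrap effects onto weapons, vehicles, characters and other item types.
# Unpacks customization-mask channels and object-state flags into named reroutes, alongside pre-skinned position, bounds and safe-zone inputs.
# Chains Galaxy, Liquid, top-side triplanar projection, time-of-day masking and split-weapon projection functions.
# MF_Base_Effects supplies the final material and decal/hotwax masks, while the projection step exports layer, emissive and triplanar masks.
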MaterialFunction /Game/Packages/Fortress_SharedMaterials/Base_Material_Functions/BaseCustomizationFunctions/MF_Base_Customization_Primary_V2
graph {nodes [
  Constant()
  FunctionInput() {bUsePreviewValueAsDefault=true, inputName=--------------------, inputType=FunctionInput_Scalar, sortPriority=40}
  FunctionInput() {inputName=Material Root, inputType=FunctionInput_MaterialAttributes, sortPriority=-200}
  FunctionInput() {bUsePreviewValueAsDefault=true, inputName=------------------, inputType=FunctionInput_Scalar, sortPriority=-19}
  n1 = MaterialFunctionCall!(
   Root_Material=NamedRerouteUsage()
   In_Material=NamedRerouteUsage()
   IsState_B=NamedRerouteUsage()
   IsState_B_V4=NamedRerouteUsage()
   IsStateS=NamedRerouteUsage()
   Caculated_Bounds=NamedRerouteUsage()
   Forward Axis=NamedRerouteUsage()
   "Msk_Ts01<--->Ts02"=NamedRerouteUsage()
   Ts_Mask_01=NamedRerouteUsage()
   Ts_Mask_02=NamedRerouteUsage()
   Ts_Emis_01_Msk=NamedRerouteUsage()
   Ts_Emis_02_Msk=NamedRerouteUsage()
   "Customization Mask (_CM RGBA)"=NamedRerouteUsage()
   Pre-Skinned Local Position=NamedRerouteUsage()
   Pre-Skinned Local Normal=NamedRerouteUsage()
   Object-SafeZone-Back=NamedRerouteUsage()
   Object-SafeZone-Middle=NamedRerouteUsage()
   Object-SafeZone-Front=NamedRerouteUsage()
   Object-SafeZone-Virtual-Pivot=NamedRerouteUsage()
   Object-SafeZone-VP_Charcter_Preview_Switch=NamedRerouteUsage()
  ) {materialFunction=&/MF_Base_Effects}
  FunctionOutput(None=@n1.Result)
  n2 = MaterialFunctionCall!(
   Material=MaterialFunctionCall!(
    Material=NamedRerouteUsage()
    IsState=NamedRerouteUsage()
    IsState_B=NamedRerouteUsage()
    CM_Msk=NamedRerouteUsage()
    Pre-Skinned Local Normal=NamedRerouteUsage()
    Calcated_Bounds=NamedRerouteUsage()
    Object-SafeZone-VP-Character-PreView-Switch=NamedRerouteUsage()
    Object-SafeZone-Virtual-Pivot=NamedRerouteUsage()
   ) {materialFunction=/Game/Weapons/WeaponSkins/Wraps/MaterialFunctions/MF_Effect_Galaxy}.Result
   IsState=NamedRerouteUsage()
   IsState_B=NamedRerouteUsage()
   CM_Msk=NamedRerouteUsage()
   Pre-Skinned Local Normal=NamedRerouteUsage()
   Calcated_Bounds=NamedRerouteUsage()
   Object-SafeZone-VP-Character-PreView-Switch=NamedRerouteUsage()
   Object-SafeZone-Virtual-Pivot=NamedRerouteUsage()
  ) {materialFunction=/Game/Weapons/WeaponSkins/Wraps/MaterialFunctions/MF_Effect_Liquid}
  n3 = MaterialFunctionCall!(
   Triplaners-Masks textures=Constant()
   Material=@n2.Result
   Root_Material=@n2.Result
   IsState=NamedRerouteUsage()
   IsState_B=NamedRerouteUsage()
   Pre-Skinned Local Normal=NamedRerouteUsage()
   Calcated_Bounds=NamedRerouteUsage()
   Local Position=NamedRerouteUsage()
   Mask Secondary=NamedRerouteUsage()
   Mask Crevise=NamedRerouteUsage()
   TS Mask_01
   TS Mask_02
   Mask Primary=NamedRerouteUsage()
   Object-SafeZone-Back=NamedRerouteUsage()
   Object-SafeZone-Middle=NamedRerouteUsage()
   Liquid Mask Input=@n2.Result
   Object-SafeZone-Front=NamedRerouteUsage()
  ) {materialFunction=/Game/Weapons/WeaponSkins/Wraps/MaterialFunctions/MF_TopSideProjection_Wrap_V02}
  FunctionOutput(None=@n3.Mask_Override) {outputName=Triplaner Mask, sortPriority=4}
  FunctionOutput(None=@n1.Hotwax_Msk) {outputName=Hotwax_Msk}
  FunctionOutput(None=NamedRerouteDeclaration(None=@n3."Blend Tops01<--->Topsside01 Msk") {name="MSk_TS1<->Ts2", nodeColor=<rgba 0 0.9531 1 1>}) {outputName="MSk_TS1<->Ts2"}
  FunctionOutput(None=@n1.Decal_Multi_Msk) {outputName=Decal_Multi_Msk}
  FunctionOutput(None=@n1.Decal_Mask_Alpha) {outputName=Decal_Mask_Alpha}
  FunctionOutput(None=NamedRerouteDeclaration(None=@n3.TS_Layer03_Mask) {name=TS_Mask_03, nodeColor=<rgba 1 0 0.7031 1>}) {outputName=TS_Mask_03}
  n4 = NamedRerouteDeclaration(None=@n3.TS_Layer01_Mask) {name=TS_Mask_01, nodeColor=<rgba 1 0 0.7031 1>}
  FunctionOutput(None=@n4) {outputName=TS_Mask_01}
  n5 = NamedRerouteDeclaration(None=@n3.TS_Layer02_Mask) {name=TS_Mask_02, nodeColor=<rgba 1 0 0.7031 1>}
  FunctionOutput(None=@n5) {outputName=TS_Mask_02}
  FunctionOutput(None=NamedRerouteDeclaration(None=@n3.Emissive 01 Msk) {name=TS_Emis_01_Msk, nodeColor=<rgba 1 0.8672 0 1>}) {outputName=TS_Emis_01_Msk}
  FunctionOutput(None=NamedRerouteDeclaration(None=@n3.Emissive 02 Msk) {name=TS_Emis_02_Msk, nodeColor=<rgba 1 0.8672 0 1>}) {outputName=TS_Emis_02_Msk}
  FunctionOutput(None=NamedRerouteDeclaration(None=@n3.TS_Layer04_Mask) {name=TS_Mask_04, nodeColor=<rgba 1 0 0.7031 1>}) {outputName=TS_Mask_04}
  FunctionOutput(None=@n1.Decal_Mask_RGBA) {outputName=Decal_RGBA}
  n6 = FunctionInput() {inputName=Material In, inputType=FunctionInput_MaterialAttributes, sortPriority=-200}
  MaterialFunctionCall!(Material=NamedRerouteDeclaration(None=@n6) {name=In_Material_Master, nodeColor=<rgba 1 0.6328 0 1>}) {materialFunction=/Engine/Functions/MaterialLayerFunctions/MatLayerBlend_BreakNormal}
  MaterialFunctionCall!(
   GMap=NamedRerouteUsage()
   TopLayerMaterial_Custom
   BottomLayerMaterial_Custom
   Crevasse=NamedRerouteUsage()
   MaskPrimary=NamedRerouteUsage()
   isMech=NamedRerouteUsage()
   isVehicle=NamedRerouteUsage()
   Material
   MaskSecondary=NamedRerouteUsage()
   MaskCombined=NamedRerouteUsage()
   PreskinnedPosition=NamedRerouteUsage()
   PreskinnedNormal=NamedRerouteUsage()
   isCharacter=NamedRerouteUsage()
   Root-Material=NamedRerouteUsage()
   IsStatsB=NamedRerouteUsage()
  ) {materialFunction=/Game/Weapons/WeaponSkins/Wraps/MaterialFunctions/MF_DirtAndEdgeWear}
  NamedRerouteDeclaration(None=FunctionInput() {bUsePreviewValueAsDefault=true, inputName=Object-SafeZone-Back, inputType=FunctionInput_Vector4, sortPriority=100}) {name=SafeZone-Back, nodeColor=<rgba 1 0.6588 0.0581 1>}
  NamedRerouteDeclaration(None=FunctionInput() {bUsePreviewValueAsDefault=true, inputName=Object-SafeZone-Middle, inputType=FunctionInput_Vector4, sortPriority=101}) {name=SafeZone-Middle, nodeColor=<rgba 1 0.6588 0.0581 1>}
  n7 = MaterialFunctionCall(Float4=NamedRerouteDeclaration(None=FunctionInput() {inputName=IsState_B_V4, inputType=FunctionInput_Vector4, sortPriority=-1}) {name=Is_StatsB_V4, nodeColor=<rgba 0 0.9253 1 1>}) {materialFunction=&BreakOutFloat4Components}
  NamedRerouteDeclaration(None=@n7.G) {name=IsPickaxe, nodeColor=<rgba 0 0.9253 1 1>}
  NamedRerouteDeclaration(None=FunctionInput() {inputName=IsState_B, inputType=FunctionInput_Vector2, sortPriority=-1}) {name=Is_StatesB, desc="R= Backpack  B=Pickaxe", nodeColor=<rgba 0 0.9253 1 1>}
  NamedRerouteDeclaration(None=FunctionInput() {bUsePreviewValueAsDefault=true, inputName=Pre-Skinned Local Position, sortPriority=41}) {name=PreSkin_Position, nodeColor=<rgba 0.6039 0.1176 0.4118 1>}
  NamedRerouteDeclaration(None=FunctionInput() {bUsePreviewValueAsDefault=true, inputName=Pre-Skinned Local Normal, sortPriority=42}) {name=PreSkin_Normal, nodeColor=<rgba 0.3451 0.3098 1 1>}
  NamedRerouteDeclaration(None=FunctionInput() {bUsePreviewValueAsDefault=true, inputName=Pre-Skinned Bounds, sortPriority=43}) {name=PreSkin_Bounds, nodeColor=<rgba 1 0.4219 0 1>}
  NamedRerouteDeclaration(None=FunctionInput() {inputName=Caculated_Bounds}) {name=Caculated_Bounds, nodeColor=<rgba 0.465 1 0.0095 1>}
  n8 = MaterialFunctionCall(Float4=NamedRerouteDeclaration(None=FunctionInput() {bUsePreviewValueAsDefault=true, inputName="Customization Mask (_CM RGBA)", inputType=FunctionInput_Vector4, sortPriority=26}) {name=Masks_RGBA, nodeColor=<rgba 0.0094 0.0094 0.901 1>}) {materialFunction=&BreakOutFloat4Components}
  NamedRerouteDeclaration(None=@n8.R) {name=Mask_Primary, nodeColor=<rgba 0.0094 0.0094 0.901 1>}
  NamedRerouteDeclaration(None=@n8.G) {name=Mask_Crevice, nodeColor=<rgba 0 0.7865 0 1>}
  NamedRerouteDeclaration(None=FunctionInput() {bUsePreviewValueAsDefault=true, inputName=Object-SafeZone-Front, inputType=FunctionInput_Vector4, sortPriority=102}) {name=SafeZone-Front, nodeColor=<rgba 1 0.6588 0.0581 1>}
  NamedRerouteDeclaration(None=@n8.B) {name=Mask_Secondary, nodeColor=<rgba 0.0094 0.0094 0.901 1>}
  NamedRerouteDeclaration(None=@n6) {name=Root_Material, nodeColor=<rgba 1 0.6328 0 1>}
  n9 = FunctionInput() {bUsePreviewValueAsDefault=true, inputName=Object-SafeZone-VP_Charcter_Preview_Switch, inputType=FunctionInput_Scalar, sortPriority=105}
  NamedRerouteDeclaration(None=@n9) {name=SafeZone-VP_Charcter_Preview_Switch, nodeColor=<rgba 1 0.6588 0.0581 1>}
  NamedRerouteDeclaration(None=@n8.A) {name=Mask_Gmap, nodeColor=<rgba 0.0094 0.0094 0.901 1>}
  NamedRerouteDeclaration(None=FunctionInput() {bUsePreviewValueAsDefault=true, inputName=Forward Axis}) {name=Forward Axis, nodeColor=<rgba 0.7813 0 1 1>}
  NamedRerouteDeclaration(None=@n8.B) {name=Mask_CM, nodeColor=<rgba 0 1 0.0859 1>}
  NamedRerouteDeclaration(None=@n7.A) {name=isCharAccent, nodeColor=<rgba 0 0.9253 1 1>}
  n10 = FunctionInput() {bUsePreviewValueAsDefault=true, inputName=Object-SafeZone-Virtual-Pivot, inputType=FunctionInput_Vector4, sortPriority=105}
  NamedRerouteDeclaration(None=@n10) {name=SafeZone-Virtual-Pivot, nodeColor=<rgba 1 0.6588 0.0581 1>}
  n11 = NamedRerouteUsage()
  NamedRerouteDeclaration(None=MaterialFunctionCall!(
    Material=MaterialFunctionCall!(Material=@n3.Result, TopSide_Mask_01=@n4, TopSide_Mask_02=@n5, "Customization Mask (_CM B)"=@n11) {materialFunction=/Game/Weapons/WeaponSkins/Wraps/MaterialFunctions/MF_TopSide_Effect_Mask_TOD}.Result
    IsState=NamedRerouteUsage()
    IsState_B=NamedRerouteUsage()
    Pre-Skinned Local Normal=NamedRerouteUsage()
    Calcated_Bounds=NamedRerouteUsage()
    Mask Crevise=NamedRerouteUsage()
    Mask Secondary=NamedRerouteUsage()
    Mask Primary=NamedRerouteUsage()
   ) {materialFunction=/Game/Weapons/WeaponSkins/Wraps/MaterialFunctions/MF_Projection_SplitWeapon}.Result) {name=Resullt, nodeColor=<rgba 1 0.75 0 1>}
  NamedRerouteDeclaration(None=@n2.Effect Liquid Msk) {name=Liquid masks, nodeColor=<rgba 1 0.8906 0 1>}
  NamedRerouteDeclaration(None=FunctionInput() {bUsePreviewValueAsDefault=true, inputName=Rotation X Axis, inputType=FunctionInput_Scalar}) {name=Rotation X, nodeColor=<rgba 0.7812 0 1 1>}
  n12 = FunctionInput() {bUsePreviewValueAsDefault=true, inputName=Wrap ScaleMultiplier, inputType=FunctionInput_Scalar, previewValue=<1 0 0 1>}
  NamedRerouteDeclaration(None=@n12) {name=Scale_Multiplier, nodeColor=<rgba 1 0 0.3516 1>}
  n13 = ScalarParameter() {group=Primary Triplaner, parameterName=Weapon Z Axis Rot, sortPriority=-34}
  n14 = ScalarParameter() {group=Primary Triplaner, parameterName=Vehicle Z Axis Rot, sortPriority=-33}
  n15 = NamedRerouteUsage()
  NamedRerouteDeclaration(None=Add(
    A=FunctionInput() {bUsePreviewValueAsDefault=true, inputName=Rotation Z Axis, inputType=FunctionInput_Scalar}
    B=LinearInterpolate(A=@n13, B=@n14, Alpha=@n15)
   )) {name=Rotation Z, nodeColor=<rgba 0.7812 0 1 1>}
  n16 = MaterialFunctionCall(Float4=NamedRerouteDeclaration(None=FunctionInput() {inputName=IsStateS, inputType=FunctionInput_Vector4, sortPriority=-1}) {name=Is_States, desc="R=vehicle G=VehicleSide =R=Mech A=Character", nodeColor=<rgba 0 0.9253 1 1>}) {materialFunction=&BreakOutFloat4Components}
  NamedRerouteDeclaration(None=@n16.B) {name=isMech, nodeColor=<rgba 0 0.9253 1 1>}
  NamedRerouteDeclaration(None=@n16.G) {name=isVehicleSide, nodeColor=<rgba 0 0.9253 1 1>}
  NamedRerouteDeclaration(None=@n16.R) {name=isVehicle, nodeColor=<rgba 0 0.9253 1 1>}
  NamedRerouteDeclaration(None=@n16.A) {name=isCharacter, nodeColor=<rgba 0 0.9253 1 1>}
  NamedRerouteDeclaration(None=@n7.R) {name=IsBackpack, nodeColor=<rgba 0 0.9253 1 1>}
  NamedRerouteUsage()
  Reroute(None=Saturate(None=Add(A=@n8.R, B=@n8.R)))
 ]}
