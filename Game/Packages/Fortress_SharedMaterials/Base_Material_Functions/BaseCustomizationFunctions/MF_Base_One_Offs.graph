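# Material function in the base customization set that gathers dozens of one-off animated and special-case wrap effects for weapons, vehicles, mechs and characters.
# Material attributes pass serially through per-wrap functions such as Storm, Holographic, TheCube and ZeroPoint, each feeding the next.
# Named reroutes share the inputs with every stage: customization mask channels, state flags, forward axis, pre-skinned position and normal, bounds, and safe-zone pivots.
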MaterialFunction /Game/Packages/Fortress_SharedMaterials/Base_Material_Functions/BaseCustomizationFunctions/MF_Base_One_Offs
graph {nodes [
  Constant()
  FunctionInput() {bUsePreviewValueAsDefault=true, inputName=--------------------, inputType=FunctionInput_Scalar, sortPriority=40}
  FunctionInput() {bUsePreviewValueAsDefault=true, inputName=------------------, inputType=FunctionInput_Scalar, sortPriority=-19}
  FunctionOutput(None=NamedRerouteUsage())
  MaterialFunctionCall!(Material=NamedRerouteDeclaration(None=FunctionInput() {inputName=Material, inputType=FunctionInput_MaterialAttributes, sortPriority=-200}) {name=Root_Material_Master, nodeColor=<rgba 1 0.6328 0 1>}) {materialFunction=/Engine/Functions/MaterialLayerFunctions/MatLayerBlend_BreakNormal}
  MaterialFunctionCall!(
   Material
   ForwardAxis=NamedRerouteUsage()
   Side Project Vehicle=NamedRerouteUsage()
   isMech=NamedRerouteUsage()
   isVehicle=NamedRerouteUsage()
   Pre-Skinned Local Normal=NamedRerouteUsage()
   Local Position=NamedRerouteUsage()
   Object Bounds Min Adjustment=NamedRerouteUsage()
   Object Bounds Max Adjustment=NamedRerouteUsage()
   Object Bounds Center Adjustment=NamedRerouteUsage()
   "Customization Mask (_CM B)"=NamedRerouteUsage()
  ) {materialFunction=/Game/Weapons/WeaponSkins/Wraps/MaterialFunctions/MF_Fryangles}
  NamedRerouteDeclaration(None=FunctionInput() {bUsePreviewValueAsDefault=true, inputName=Object-SafeZone-Back, inputType=FunctionInput_Vector4, sortPriority=100}) {name=SafeZone-Back, nodeColor=<rgba 1 0.6588 0.0581 1>}
  NamedRerouteDeclaration(None=FunctionInput() {bUsePreviewValueAsDefault=true, inputName=Object-SafeZone-Middle, inputType=FunctionInput_Vector4, sortPriority=101}) {name=SafeZone-Middle, nodeColor=<rgba 1 0.6588 0.0581 1>}
  n1 = MaterialFunctionCall(Float4=NamedRerouteDeclaration(None=FunctionInput() {inputName=IsState_B_V4, inputType=FunctionInput_Vector4, sortPriority=-1}) {name=Is_StatesB_V4, nodeColor=<rgba 0 0.9253 1 1>}) {materialFunction=&BreakOutFloat4Components}
  NamedRerouteDeclaration(None=@n1.G) {name=IsPickaxe, nodeColor=<rgba 0 0.9253 1 1>}
  NamedRerouteDeclaration(None=FunctionInput() {inputName=IsState_B, inputType=FunctionInput_Vector2, sortPriority=-1}) {name=Is_StatesB, desc="R= Backpack  B=Pickaxe", nodeColor=<rgba 0 0.9253 1 1>}
  n2 = MaterialFunctionCall(Float4=NamedRerouteDeclaration(None=FunctionInput() {bUsePreviewValueAsDefault=true, inputName="Customization Mask (_CM RGBA)", inputType=FunctionInput_Vector4, sortPriority=26}) {name=Masks_RGBA, nodeColor=<rgba 0.0094 0.0094 0.901 1>}) {materialFunction=&BreakOutFloat4Components}
  NamedRerouteDeclaration(None=Saturate(None=Add(A=@n2.R, B=@n2.R))) {name=Mask_Primary+Secondary, nodeColor=<rgba 0.0094 0.0094 0.901 1>}
  NamedRerouteDeclaration(None=FunctionInput() {bUsePreviewValueAsDefault=true, inputName=Bounds_Extent}) {name=Bounds_Extent, nodeColor=<rgba 0.465 1 0.0095 1>}
  NamedRerouteDeclaration(None=FunctionInput() {bUsePreviewValueAsDefault=true, inputName=Pre-Skinned Local Position, sortPriority=41}) {name=PreSkin_Position, nodeColor=<rgba 0.6039 0.1176 0.4118 1>}
  NamedRerouteDeclaration(None=FunctionInput() {bUsePreviewValueAsDefault=true, inputName=Pre-Skinned Local Normal, sortPriority=42}) {name=PreSkin_Normal, nodeColor=<rgba 0.3451 0.3098 1 1>}
  NamedRerouteDeclaration(None=FunctionInput() {bUsePreviewValueAsDefault=true, inputName=Pre-Skinned Bounds, sortPriority=43}) {name=PreSkin_Bounds, nodeColor=<rgba 1 0.4219 0 1>}
  NamedRerouteDeclaration(None=FunctionInput() {bUsePreviewValueAsDefault=true, inputName=Caculated_Bounds}) {name=Caculated_Bounds, nodeColor=<rgba 0.465 1 0.0095 1>}
  NamedRerouteDeclaration(None=@n2.R) {name=Mask_Primary, nodeColor=<rgba 0.0094 0.0094 0.901 1>}
  NamedRerouteDeclaration(None=@n2.G) {name=Mask_Crevice, nodeColor=<rgba 0 0.7865 0 1>}
  NamedRerouteDeclaration(None=FunctionInput() {bUsePreviewValueAsDefault=true, inputName=Object-SafeZone-Front, inputType=FunctionInput_Vector4, sortPriority=102}) {name=SafeZone-Front, nodeColor=<rgba 1 0.6588 0.0581 1>}
  NamedRerouteDeclaration(None=@n2.B) {name=Mask_Secondary, nodeColor=<rgba 0.0094 0.0094 0.901 1>}
  NamedRerouteDeclaration(None=FunctionInput() {bUsePreviewValueAsDefault=true, inputName=Object Bounds Max Adjustment}) {name=Bounds_Max, nodeColor=<rgba 1 0 0.4811 1>}
  NamedRerouteDeclaration(None=FunctionInput() {bUsePreviewValueAsDefault=true, inputName=Object SBounds Min Adjustment}) {name=Bounds_Min, nodeColor=<rgba 1 0 0.4811 1>}
  n3 = FunctionInput() {bUsePreviewValueAsDefault=true, inputName=Object-SafeZone-VP_Charcter_Preview_Switch, inputType=FunctionInput_Scalar, sortPriority=105}
  NamedRerouteDeclaration(None=@n3) {name=SafeZone-VP_Charcter_Preview_Switch, nodeColor=<rgba 1 0.6588 0.0581 1>}
  NamedRerouteDeclaration(None=@n2.A) {name=Mask_Gmap, nodeColor=<rgba 0.0094 0.0094 0.901 1>}
  NamedRerouteDeclaration(None=FunctionInput() {bUsePreviewValueAsDefault=true, inputName=Forward Axis}) {name=Forward Axis, nodeColor=<rgba 0.7812 0 1 1>}
  NamedRerouteDeclaration(None=@n2.B) {name=Mask_CM, nodeColor=<rgba 0 1 0.0859 1>}
  NamedRerouteDeclaration(None=@n1.A) {name=isCharAccent, nodeColor=<rgba 0 0.9253 1 1>}
  n4 = FunctionInput() {bUsePreviewValueAsDefault=true, inputName=Object-SafeZone-Virtual-Pivot, inputType=FunctionInput_Vector4, sortPriority=105}
  NamedRerouteDeclaration(None=@n4) {name=SafeZone-Virtual-Pivot, nodeColor=<rgba 1 0.6588 0.0581 1>}
  n5 = NamedRerouteUsage()
  n6 = NamedRerouteUsage()
  n7 = NamedRerouteUsage()
  n8 = NamedRerouteUsage()
  n9 = NamedRerouteUsage()
  n10 = NamedRerouteUsage()
  n11 = NamedRerouteUsage()
  n12 = NamedRerouteUsage()
  n13 = NamedRerouteUsage()
  n14 = NamedRerouteUsage()
  n15 = NamedRerouteUsage()
  n16 = NamedRerouteUsage()
  n17 = NamedRerouteUsage()
  n18 = NamedRerouteUsage()
  n19 = NamedRerouteUsage()
  n20 = NamedRerouteUsage()
  n21 = NamedRerouteUsage()
  n22 = NamedRerouteUsage()
  n23 = NamedRerouteUsage()
  n24 = MaterialFunctionCall!(
   Material=MaterialFunctionCall!(
    isCharacter=NamedRerouteUsage()
    Material=@n23
    ForwardAxis=NamedRerouteUsage()
    Material_Bypass=@n23
    isMech=NamedRerouteUsage()
    isVehicle=NamedRerouteUsage()
    Pre-Skinned Local Normal=NamedRerouteUsage()
    Local Position=NamedRerouteUsage()
    Object Bounds Min Adjustment=NamedRerouteUsage()
    Object Bounds Max Adjustment=NamedRerouteUsage()
    Object Bounds Center Adjustment=NamedRerouteUsage()
    "Customization Mask (_CM B)"=NamedRerouteUsage()
    IsStatsB=NamedRerouteUsage()
   ) {materialFunction=/Game/Weapons/WeaponSkins/Wraps/MaterialFunctions/MF_ActivatedRunes_Wrap}.Result
   IsVehicle=NamedRerouteUsage()
   IsStatsB=NamedRerouteUsage()
   IsCharacter=NamedRerouteUsage()
   ForwardAxis=NamedRerouteUsage()
   Wrap ScaleMultiplier=NamedRerouteUsage()
   PreSkinLocalPosition=NamedRerouteUsage()
   PreSkinnedLocalNormal=NamedRerouteUsage()
   "Customization Mask (_CM B)"=NamedRerouteUsage()
  ) {materialFunction=/Game/Weapons/WeaponSkins/Wraps/MaterialFunctions/MF_Customization_Storm}
  n25 = NamedRerouteUsage()
  n26 = NamedRerouteUsage()
  n27 = NamedRerouteUsage()
  n28 = NamedRerouteUsage()
  n29 = NamedRerouteUsage()
  n30 = NamedRerouteUsage()
  n31 = NamedRerouteUsage()
  n32 = NamedRerouteUsage()
  n33 = MaterialFunctionCall!(
   Material=MaterialFunctionCall!(
    Material=MaterialFunctionCall(
     Material=@n24.Result
     IsVehicle=@n25
     IsStatsB=@n26
     IsCharacter=@n27
     ForwardAxis=@n28
     Wrap ScaleMultiplier=@n29
     PreSkinLocalPosition=@n30
     PreSkinnedLocalNormal=@n31
     "Customization Mask (_CM B)"=@n32
    ) {materialFunction=&MF_Customization_Storm}.Result
    isVehicle=NamedRerouteUsage()
    ForwardAxis=NamedRerouteUsage()
    Pre-Skinned Local Position=NamedRerouteUsage()
   ) {materialFunction=/Game/Weapons/WeaponSkins/Wraps/MaterialFunctions/MF_Customization_Holographic}.Result
   Wrap ScaleMultiplier=NamedRerouteUsage()
   ForwardAxis=NamedRerouteUsage()
   "Customization Mask (_CM B)"=NamedRerouteUsage()
   Pre-Skinned Bounds=NamedRerouteUsage()
   Pre-Skinned Local Position=NamedRerouteUsage()
   Pre-Skinned Local Normal=NamedRerouteUsage()
  ) {materialFunction=/Game/Weapons/WeaponSkins/Wraps/MaterialFunctions/MF_Customization_Shatter}
  n34 = NamedRerouteUsage()
  n35 = NamedRerouteUsage()
  n36 = NamedRerouteUsage()
  n37 = NamedRerouteUsage()
  n38 = NamedRerouteUsage()
  n39 = NamedRerouteUsage()
  n40 = NamedRerouteUsage()
  n41 = NamedRerouteUsage()
  n42 = MaterialFunctionCall!(
   Material=MaterialFunctionCall!(
    Material=MaterialFunctionCall!(
     Material=@n33.Result
     Local Position=@n34
     isVehicle=@n35
     isMech=@n36
     Object Bounds Max Adjustment=@n37
     Object Bounds Min Adjustment=@n38
     Object Bounds Center Adjustment=@n39
     Pre-Skinned Local Normal=@n40
     "Customization Mask (_CM B)"=@n41
    ) {materialFunction=/Game/Weapons/WeaponSkins/Wraps/MaterialFunctions/MF_Wrap_Equalizer}.Result
    Pre-Skinned Local Position=NamedRerouteUsage()
    RotXAxis=NamedRerouteUsage()
    isVehicle=NamedRerouteUsage()
    IsCharacter=NamedRerouteUsage()
    IsStatsB=NamedRerouteUsage()
    RotZAxis=NamedRerouteUsage()
    IsStats=NamedRerouteUsage()
    Calcated_Bounds=NamedRerouteUsage()
    "Customization Crevasse (_CM G)"=NamedRerouteUsage()
    "Customization Mask (_CM B)"=NamedRerouteUsage()
    Pre-Skinned Local Normal=NamedRerouteUsage()
    Pre-Skinned Bounds=NamedRerouteUsage()
    Object-SafeZone-VP-Character-PreView-Switch=NamedRerouteUsage()
    Object-SafeZone-Virtual-Pivot=NamedRerouteUsage()
   ) {materialFunction=/Game/Weapons/WeaponSkins/Wraps/MaterialFunctions/MF_Customization_TheCube}.Result
   IsCharacter=NamedRerouteUsage()
   IsStatsB=NamedRerouteUsage()
   IsVehicle=NamedRerouteUsage()
   Wrap ScaleMultiplier=NamedRerouteUsage()
   ForwardAxis=NamedRerouteUsage()
   PreSkinnedLocalNormal=NamedRerouteUsage()
   PreSkinLocalPosition=NamedRerouteUsage()
  ) {materialFunction=/Game/Weapons/WeaponSkins/Wraps/MaterialFunctions/MF_Customization_WeirdObjects}
  n43 = NamedRerouteUsage()
  n44 = NamedRerouteUsage()
  n45 = MaterialFunctionCall!(
   Material=MaterialFunctionCall!(
    Material=MaterialFunctionCall!(Material=@n42.Result, isVehicle=@n43, "Customization Mask (_CM B)"=@n44) {materialFunction=/Game/Weapons/WeaponSkins/Wraps/MaterialFunctions/MF_ZeroPointWrap}.Result
    IsStatsB=NamedRerouteUsage()
    Side Project Vehicle=NamedRerouteUsage()
    isCharacter=NamedRerouteUsage()
    isMech=NamedRerouteUsage()
    isVehicle=NamedRerouteUsage()
    "Customization Crevasse (_CM G)"=NamedRerouteUsage()
    Customization Mask B=NamedRerouteUsage()
    "Customization Mask (_CM B)"=NamedRerouteUsage()
    Object Bounds Min Adjustment=NamedRerouteUsage()
    Object Bounds Max Adjustment=NamedRerouteUsage()
    Object-SafeZone-Virtual-CH-Switch=NamedRerouteUsage()
    Object-SafeZone-Virtual-Pivot=NamedRerouteUsage()
   ) {materialFunction=/Game/Weapons/WeaponSkins/Wraps/MaterialFunctions/MF_ZeroPointWrap2}.Result
   isMech=NamedRerouteUsage()
   IsStatsB=NamedRerouteUsage()
   IsCharacter=NamedRerouteUsage()
   "Customization Crevasse (_CM G)"=NamedRerouteUsage()
   "Customization Mask (_CM B)"=NamedRerouteUsage()
   Object Bounds Min Adjustment=NamedRerouteUsage()
   Object Bounds Max Adjustment=NamedRerouteUsage()
   Object-SafeZone-Virtual-CH-Switch=NamedRerouteUsage()
   Object-SafeZone-Virtual-Pivot=NamedRerouteUsage()
  ) {materialFunction=/Game/Weapons/WeaponSkins/Wraps/MaterialFunctions/MF_ZeroPointWrap3}
  n46 = NamedRerouteUsage()
  n47 = NamedRerouteUsage()
  n48 = NamedRerouteUsage()
  n49 = NamedRerouteUsage()
  n50 = NamedRerouteUsage()
  n51 = MaterialFunctionCall!(
   In=MaterialFunctionCall!(
    IsVehicle=NamedRerouteUsage()
    IsMech=NamedRerouteUsage()
    Pre-SkinnedLocalPosition=NamedRerouteUsage()
    Pre-SkinnedLocalBounds
    Pre-SkinnedLocalBounds_Min
    MaterialInput=MaterialFunctionCall!(
     PreskinnedPosition=@n19
     PreskinnedBounds=@n20
     isVehicle=@n21
     isMech=@n22
     Material=@n45.Result
     IsStatsB=@n46
     IsCharacter=@n47
     "Customization Mask (_CM B)"=@n48
     Object Bounds Min Adjustment=@n49
     Object Bounds Max Adjustment=@n50
    ) {materialFunction=/Game/Weapons/WeaponSkins/MaterialFunction/MF_HauntWrap}.Result
    IsStatsB=NamedRerouteUsage()
   ) {materialFunction=/Game/Weapons/WeaponSkins/Wraps/MaterialFunctions/MF_TechOpsBlue}.Result
   Mask=NamedRerouteUsage()
   isVehicle=NamedRerouteUsage()
   isMech=NamedRerouteUsage()
  ) {materialFunction=/Game/Characters/Player/Male/Medium/Bodies/M_MED_Mastermind/Materials/MF_TwistedGoo}
  n52 = NamedRerouteUsage()
  n53 = NamedRerouteUsage()
  n54 = NamedRerouteUsage()
  n55 = NamedRerouteUsage()
  n56 = NamedRerouteUsage()
  n57 = MaterialFunctionCall!(
   Material=MaterialFunctionCall!(
    Mask=NamedRerouteUsage()
    isMech=NamedRerouteUsage()
    isVehicle=NamedRerouteUsage()
    Material=MaterialFunctionCall!(
     Local Position=@n13
     isMech=@n14
     isVehicle=@n15
     Object Bounds Max Adjustment=@n16
     Object Bounds Min Adjustment=@n17
     Object Bounds Center Adjustment=@n18
     In=@n51.Result
     Alpha=@n52
     Local normal=@n53
     Side Project Vehicle=@n54
     PreSkin_Bounds=@n55
     Bounds_Extents=@n56
    ) {materialFunction=/Game/Weapons/WeaponSkins/Wraps/MaterialFunctions/MF_NeonAnimalWrap}.Result
   ) {materialFunction=/Game/Weapons/WeaponSkins/Wraps/MaterialFunctions/MF_IceBreakerWrap}.Result
   Side Project Vehicle=NamedRerouteUsage()
   IsMech=NamedRerouteUsage()
   IsVehicle=NamedRerouteUsage()
   Wrap Bounds=NamedRerouteUsage()
   "Customization Crevasse (_CM G)"=NamedRerouteUsage()
   IsStatsB=NamedRerouteUsage()
   Object-SafeZone-VP-Character-PreView-Switch=NamedRerouteUsage()
   Object-SafeZone-Virtual-Pivot=NamedRerouteUsage()
  ) {materialFunction=/Game/Weapons/WeaponSkins/Wraps/MaterialFunctions/MF_ConstellationWrap}
  n58 = NamedRerouteUsage()
  n59 = NamedRerouteUsage()
  n60 = NamedRerouteUsage()
  n61 = NamedRerouteUsage()
  n62 = NamedRerouteUsage()
  n63 = NamedRerouteUsage()
  n64 = MaterialFunctionCall!(
   Alpha Mask=NamedRerouteUsage()
   GMAP=NamedRerouteUsage()
   In=MaterialFunctionCall!(
    Pre-SkinLocalPos=NamedRerouteUsage()
    IsCharacter=NamedRerouteUsage()
    IsStatsB=NamedRerouteUsage()
    In=MaterialFunctionCall!(
     Material=@n57.Result
     isCharacter=@n58
     Local Position=@n59
     IsMech=@n60
     IsVehicle=@n61
     "Customization Crevasse (_CM G)"=@n62
     IsStatsB=@n63
    ) {materialFunction=/Game/Weapons/WeaponSkins/Wraps/MaterialFunctions/MF_StarsWrap}.Result
    Pre-SkinLocalNormal=NamedRerouteUsage()
    Pre-SkinBounds=NamedRerouteUsage()
    CMask=NamedRerouteUsage()
    IsVehicle=NamedRerouteUsage()
    IsMech=NamedRerouteUsage()
   ) {materialFunction=/Game/Weapons/WeaponSkins/Wraps/MaterialFunctions/MF_BannerWrap}.Result
  ) {materialFunction=&/MF_JacketWrap}
  n65 = NamedRerouteUsage()
  n66 = NamedRerouteUsage()
  n67 = NamedRerouteUsage()
  n68 = NamedRerouteUsage()
  n69 = NamedRerouteUsage()
  n70 = NamedRerouteUsage()
  n71 = NamedRerouteUsage()
  n72 = NamedRerouteUsage()
  n73 = NamedRerouteUsage()
  n74 = MaterialFunctionCall!(
   Material=MaterialFunctionCall!(
    isMech=NamedRerouteUsage()
    Material=MaterialFunctionCall!(
     Side Project Vehicle=@n12
     In=@n64.Result
     ForwardAxis=@n65
     isMech=@n66
     Local Normals=@n67
     Local Position=@n68
     isVehicle=@n69
     Object Bounds Center Adjustment=@n70
     Object Bounds Min Adjustment=@n71
     Object Bounds Max Adjustment=@n72
     "Customization Mask (_CM B)"=@n73
    ) {materialFunction=/Game/Weapons/WeaponSkins/Wraps/MaterialFunctions/MF_Carnaval}.Result
    Side Project Vehicle=NamedRerouteUsage()
    Local Normals=NamedRerouteUsage()
    Local Position=NamedRerouteUsage()
    isVehicle=NamedRerouteUsage()
    isCharacter=NamedRerouteUsage()
    Object Bounds Min Adjustment=NamedRerouteUsage()
    Object Bounds Center Adjustment=NamedRerouteUsage()
    Object Bounds Max Adjustment=NamedRerouteUsage()
    Customization Mask B=NamedRerouteUsage()
    "Customization Mask (_CM B)"=NamedRerouteUsage()
   ) {materialFunction=/Game/Weapons/WeaponSkins/Wraps/MaterialFunctions/MF_CarnavalCycle}.Result
   Local Position=NamedRerouteUsage()
   Local Normals=NamedRerouteUsage()
   Side Project Vehicle=NamedRerouteUsage()
   ForwardAxis=NamedRerouteUsage()
   isVehicle=NamedRerouteUsage()
   isMech=NamedRerouteUsage()
   Object Bounds Center Adjustment=NamedRerouteUsage()
   Customization Mask Primary=NamedRerouteUsage()
   Customization Mask Curvatur=NamedRerouteUsage()
   Object Bounds Min Adjustment=NamedRerouteUsage()
   Object Bounds Max Adjustment=NamedRerouteUsage()
  ) {materialFunction=/Game/Weapons/WeaponSkins/Wraps/MaterialFunctions/MF_Spectrum}
  n75 = NamedRerouteUsage()
  n76 = NamedRerouteUsage()
  n77 = NamedRerouteUsage()
  n78 = NamedRerouteUsage()
  n79 = NamedRerouteUsage()
  n80 = MaterialFunctionCall!(
   ForwardAxis=NamedRerouteUsage()
   Material=MaterialFunctionCall!(
    Material=MaterialFunctionCall!(
     ObjectBounds_Center=@n6
     Mask=@n7
     PreskionnedPosition=@n8
     isVehicle=@n9
     isMech=@n10
     PreskinnedNormal=@n11
     Material=@n74.Result
     ObjectBounds_Min=@n75
     ObjectBounds_Max=@n76
     Vehicle Side=@n77
     isCharacter=@n78
     IsStatsB=@n79
    ) {materialFunction=/Game/Weapons/WeaponSkins/Wraps/MaterialFunctions/MF_TargetPractice_Wrap}.Result
    isMech=NamedRerouteUsage()
    isVehicle=NamedRerouteUsage()
    Pre-Skinned Local Normal=NamedRerouteUsage()
    Local Position=NamedRerouteUsage()
    Object Bounds Min Adjustment=NamedRerouteUsage()
    Object Bounds Max Adjustment=NamedRerouteUsage()
    Object Bounds Center Adjustment=NamedRerouteUsage()
    Customization Mask Accent=NamedRerouteUsage()
    "Customization Mask (_CM B)"=NamedRerouteUsage()
   ) {materialFunction=/Game/Weapons/WeaponSkins/Wraps/MaterialFunctions/MF_PeeledFoil}.Result
   isMech=NamedRerouteUsage()
   isVehicle=NamedRerouteUsage()
   Pre-Skinned Local Normal=NamedRerouteUsage()
   Local Position=NamedRerouteUsage()
   Object Bounds Min Adjustment=NamedRerouteUsage()
   Object Bounds Max Adjustment=NamedRerouteUsage()
   Object Bounds Center Adjustment=NamedRerouteUsage()
   "Customization Mask (_CM B)"=NamedRerouteUsage()
  ) {materialFunction=/Game/Weapons/WeaponSkins/Wraps/MaterialFunctions/MF_NFWrap}
  n81 = NamedRerouteUsage()
  n82 = NamedRerouteUsage()
  n83 = NamedRerouteUsage()
  n84 = NamedRerouteUsage()
  n85 = NamedRerouteUsage()
  n86 = NamedRerouteUsage()
  n87 = NamedRerouteUsage()
  n88 = NamedRerouteUsage()
  n89 = MaterialFunctionCall!(
   Material=MaterialFunctionCall!(
    Material=MaterialFunctionCall!(
     Material=@n80.Result
     isMech=@n81
     isVehicle=@n82
     Pre-Skinned Local Normal=@n83
     Local Position=@n84
     Object Bounds Min Adjustment=@n85
     Object Bounds Max Adjustment=@n86
     Object Bounds Center Adjustment=@n87
     "Customization Mask (_CM B)"=@n88
    ) {materialFunction=/Game/Weapons/WeaponSkins/Wraps/MaterialFunctions/MF_OceanRider_Wrap}.Result
    isMech=NamedRerouteUsage()
    isVehicle=NamedRerouteUsage()
    Pre-Skinned Local Normal=NamedRerouteUsage()
    Local Position=NamedRerouteUsage()
    Object Bounds Min Adjustment=NamedRerouteUsage()
    Object Bounds Max Adjustment=NamedRerouteUsage()
    Object Bounds Center Adjustment=NamedRerouteUsage()
    "Customization Crevasse (_CM G)"=NamedRerouteUsage()
    "Customization Mask (_CM B)"=NamedRerouteUsage()
   ) {materialFunction=/Game/Weapons/WeaponSkins/Wraps/MaterialFunctions/MF_GlowVortex_Wrap}.Result
   Pre-Skinned Local Normal=NamedRerouteUsage()
   Local Position=NamedRerouteUsage()
   ForwardAxis=NamedRerouteUsage()
   Cal_Bounds=NamedRerouteUsage()
   isCharacter=NamedRerouteUsage()
   Side Project Vehicle=NamedRerouteUsage()
   isMech=NamedRerouteUsage()
   isVehicle=NamedRerouteUsage()
   "Customization Mask (_CM B)"=NamedRerouteUsage()
  ) {materialFunction=/Game/Weapons/WeaponSkins/Wraps/MaterialFunctions/MF_WrapRainbowGlitter}
  n90 = NamedRerouteUsage()
  n91 = NamedRerouteUsage()
  n92 = NamedRerouteUsage()
  n93 = NamedRerouteUsage()
  n94 = NamedRerouteUsage()
  n95 = NamedRerouteUsage()
  n96 = NamedRerouteUsage()
  n97 = NamedRerouteUsage()
  n98 = NamedRerouteUsage()
  n99 = NamedRerouteUsage()
  n100 = MaterialFunctionCall!(
   PreSkin_Bounds=NamedRerouteUsage()
   Side Project Vehicle=NamedRerouteUsage()
   isMech=NamedRerouteUsage()
   In=MaterialFunctionCall!(
    Material=MaterialFunctionCall!(
     Material_input=@n89.Result
     Crevice=@n90
     isMech=@n91
     Side Project Vehicle=@n92
     isVehicle=@n93
     isCharacter=@n94
     Customization Mask=@n95
     Object Bounds Min Adjustment=@n96
     Object Bounds Max Adjustment=@n97
     Object-SafeZone-Virtual-CH-Switch=@n98
     Object-SafeZone-Virtual-Pivot=@n99
    ) {materialFunction=/Game/Weapons/WeaponSkins/Wraps/MaterialFunctions/MF_WrapNeonPulse}.Result
    isMech=NamedRerouteUsage()
    isVehicle=NamedRerouteUsage()
    Pre-Skinned Local Normal=NamedRerouteUsage()
    Local Position=NamedRerouteUsage()
    Object Bounds Min Adjustment=NamedRerouteUsage()
    Object Bounds Max Adjustment=NamedRerouteUsage()
    Object Bounds Center Adjustment=NamedRerouteUsage()
    "Customization Mask (_CM B)"=NamedRerouteUsage()
   ) {materialFunction=/Game/Weapons/WeaponSkins/Wraps/MaterialFunctions/MF_GlowCamo_Wrap}.Result
   isVehicle=NamedRerouteUsage()
   "Customization Mask (_CM B)"=NamedRerouteUsage()
  ) {materialFunction=/Game/Weapons/WeaponSkins/Wraps/MaterialFunctions/MF_Ultravioletwrap.MF_UltravioletWrap}
  n101 = NamedRerouteUsage()
  n102 = NamedRerouteUsage()
  n103 = NamedRerouteUsage()
  n104 = NamedRerouteUsage()
  n105 = NamedRerouteUsage()
  n106 = MaterialFunctionCall!(
   Material=MaterialFunctionCall!(
    Material=MaterialFunctionCall!(
     Material=@n100.Result
     Side Project Vehicle=@n101
     Warp Mask
     IsMech=@n102
     IsVehicle=@n103
     Wrap Bounds=@n104
     "Customization Crevasse (_CM G)"=@n105
    ) {materialFunction=/Game/Weapons/WeaponSkins/Wraps/MaterialFunctions/MF_Tapas_Wrap}.Result
    Side Project Vehicle=NamedRerouteUsage()
    IsCharacter=NamedRerouteUsage()
    isMech=NamedRerouteUsage()
    isVehicle=NamedRerouteUsage()
    Calculated Bounds=NamedRerouteUsage()
    Pre-Skinned Local Normal=NamedRerouteUsage()
    Local Position=NamedRerouteUsage()
    Object Bounds Min Adjustment=NamedRerouteUsage()
    Wrap Bounds
    Object Bounds Max Adjustment=NamedRerouteUsage()
    Object Bounds Center Adjustment=NamedRerouteUsage()
    "Customization Mask 2 (_CM B Top Right)"=NamedRerouteUsage()
    isCharAccent=NamedRerouteUsage()
   ) {materialFunction=/Game/Weapons/WeaponSkins/Wraps/MaterialFunctions/MF_HoneyDew_Wrap}.Result
   isVehicle=NamedRerouteUsage()
   isMech=NamedRerouteUsage()
   Side Project Vehicle=NamedRerouteUsage()
   ForwardAxis=NamedRerouteUsage()
   Local Position=NamedRerouteUsage()
   Object Bounds Min Adjustment=NamedRerouteUsage()
   Object Bounds Max Adjustment=NamedRerouteUsage()
   Object Bounds Center Adjustment=NamedRerouteUsage()
   Pre-Skinned Local Normal=NamedRerouteUsage()
   "Customization Mask (_CM B)"=NamedRerouteUsage()
  ) {materialFunction=/Game/Weapons/WeaponSkins/Wraps/MaterialFunctions/MF_Date_Wrap}
  n107 = NamedRerouteUsage()
  n108 = NamedRerouteUsage()
  n109 = NamedRerouteUsage()
  n110 = NamedRerouteUsage()
  n111 = NamedRerouteUsage()
  n112 = NamedRerouteUsage()
  n113 = NamedRerouteUsage()
  n114 = MaterialFunctionCall!(
   Material=MaterialFunctionCall!(
    Material=MaterialFunctionCall!(
     Material=@n106.Result
     Side Project Vehicle=@n107
     isMech=@n108
     isVehicle=@n109
     Pre-Skinned Local Normal=@n110
     Calcated_Bounds=@n111
     Customization Mask Curvature=@n112
     "Customization Mask (_CM B)"=@n113
    ) {materialFunction=/Game/Weapons/WeaponSkins/Wraps/MaterialFunctions/MF_Wrap_Subwoofer}.Result
    isVehicle=NamedRerouteUsage()
    isMech=NamedRerouteUsage()
    IsCharacter=NamedRerouteUsage()
    Side Project Vehicle=NamedRerouteUsage()
    ForwardAxis=NamedRerouteUsage()
    Local Position=NamedRerouteUsage()
    Object Bounds Min Adjustment=NamedRerouteUsage()
    Object Bounds Max Adjustment=NamedRerouteUsage()
    Object Bounds Center Adjustment=NamedRerouteUsage()
    Pre-Skinned Local Normal=NamedRerouteUsage()
    "Customization Mask (_CM B)"=NamedRerouteUsage()
    "Customization Mask 2 (_CM B Top Right)"=NamedRerouteUsage()
    isCharAccent=NamedRerouteUsage()
   ) {materialFunction=/Game/Weapons/WeaponSkins/Wraps/MaterialFunctions/MF_Tomato_Wrap}.Result
   isVehicle=NamedRerouteUsage()
   Pre-Skinned Local Position=NamedRerouteUsage()
   IsStatsB=NamedRerouteUsage()
  ) {materialFunction=/Game/Weapons/WeaponSkins/Wraps/MaterialFunctions/MF_Soy_Wrap}
  n115 = NamedRerouteUsage()
  n116 = NamedRerouteUsage()
  n117 = NamedRerouteUsage()
  n118 = NamedRerouteUsage()
  n119 = NamedRerouteUsage()
  n120 = NamedRerouteUsage()
  n121 = NamedRerouteUsage()
  n122 = NamedRerouteUsage()
  n123 = NamedRerouteUsage()
  n124 = NamedRerouteUsage()
  n125 = NamedRerouteUsage()
  n126 = MaterialFunctionCall!(
   Material=MaterialFunctionCall!(
    Material=MaterialFunctionCall!(
     Material=@n114.Result
     ForwardAxis=@n115
     IsCharacter=@n116
     isMech=@n117
     isVehicle=@n118
     Pre-Skinned Local Normal=@n119
     Calcated_Bounds=@n120
     "Customization Mask (_CM B)"=@n121
     Object Bounds Max Adjustment=@n122
     Object Bounds Min Adjustment=@n123
     Object-SafeZone-Virtual-CH-Switch=@n124
     Object-SafeZone-Virtual-Pivot=@n125
    ) {materialFunction=/Game/Weapons/WeaponSkins/Wraps/MaterialFunctions/MF_Wrap_Squash}.Result
    ForwardAxis=NamedRerouteUsage()
    Side Project Vehicle=NamedRerouteUsage()
    isMech=NamedRerouteUsage()
    isVehicle=NamedRerouteUsage()
    Pre-Skinned Local Normal=NamedRerouteUsage()
    Calcated_Bounds=NamedRerouteUsage()
    "Customization Mask (_CM B)"=NamedRerouteUsage()
   ) {materialFunction=/Game/Weapons/WeaponSkins/Wraps/MaterialFunctions/MF_Wrap_NeonJellyFish}.Result
   isCharacter=NamedRerouteUsage()
   isMech=NamedRerouteUsage()
   isVehicle=NamedRerouteUsage()
   Pre-Skinned Local Normal=NamedRerouteUsage()
   Calcated_Bounds=NamedRerouteUsage()
   "Customization Mask (_CM B Top left)"=NamedRerouteUsage()
   "Customization Mask 2 (_CM B Top Right)"=NamedRerouteUsage()
  ) {materialFunction=/Game/Weapons/WeaponSkins/Wraps/MaterialFunctions/MF_NeonSign_Wrap}
  n127 = NamedRerouteUsage()
  n128 = NamedRerouteUsage()
  n129 = NamedRerouteUsage()
  n130 = NamedRerouteUsage()
  n131 = NamedRerouteUsage()
  n132 = NamedRerouteUsage()
  n133 = NamedRerouteUsage()
  n134 = NamedRerouteUsage()
  n135 = MaterialFunctionCall!(
   Material=MaterialFunctionCall!(
    Material=MaterialFunctionCall!(
     Material=@n126.Result
     isMech=@n127
     isVehicleSide=@n128
     isVehicle=@n129
     ForwardAxis=@n130
     Pre-Skinned Local Normal=@n131
     Calculated_Bounds=@n132
     "Customization Mask (_CM B Top left)"=@n133
     "Customization Mask 2 (_CM B Top Right)"=@n134
    ) {materialFunction=/Game/Weapons/WeaponSkins/Wraps/MaterialFunctions/MF_Birthday2020_Wrap}.Result
    Side Project Vehicle=NamedRerouteUsage()
    ForwardAxis=NamedRerouteUsage()
    isMech=NamedRerouteUsage()
    isVehicle=NamedRerouteUsage()
    Pre-Skinned Local Normal=NamedRerouteUsage()
    Calcated_Bounds=NamedRerouteUsage()
    "Customization Mask (_CM B)"=NamedRerouteUsage()
   ) {materialFunction=/Game/Weapons/WeaponSkins/Wraps/MaterialFunctions/MF_Wrap_Muzak}.Result
   ForwardAxis=NamedRerouteUsage()
   isMech=NamedRerouteUsage()
   isVehicle=NamedRerouteUsage()
   Side Project Vehicle=NamedRerouteUsage()
   Pre-Skinned Local Normal=NamedRerouteUsage()
   Local Position=NamedRerouteUsage()
   Object Bounds Min Adjustment=NamedRerouteUsage()
   Object Bounds Max Adjustment=NamedRerouteUsage()
   Object Bounds Center Adjustment=NamedRerouteUsage()
   "Customization Mask (_CM B)"=NamedRerouteUsage()
  ) {materialFunction=/Game/Weapons/WeaponSkins/Wraps/MaterialFunctions/MF_KevinCouture_Wrap}
  n136 = NamedRerouteUsage()
  n137 = NamedRerouteUsage()
  n138 = NamedRerouteUsage()
  n139 = NamedRerouteUsage()
  n140 = NamedRerouteUsage()
  n141 = NamedRerouteUsage()
  n142 = NamedRerouteUsage()
  n143 = NamedRerouteUsage()
  n144 = NamedRerouteUsage()
  n145 = MaterialFunctionCall!(
   Material=MaterialFunctionCall!(
    Material=MaterialFunctionCall!(
     Material=@n135.Result
     ForwardAxis=@n136
     isMech=@n137
     isVehicle=@n138
     Side Project Vehicle=@n139
     Pre-Skinned Local Normal=@n140
     Local Position=@n141
     Wrap Bounds=@n142
     "Customization Mask (_CM B Top left)"=@n143
     "Customization Mask 2 (_CM B Top Right)"=@n144
    ) {materialFunction=/Game/Weapons/WeaponSkins/Wraps/MaterialFunctions/MF_Beats_Wrap}.Result
    Side Project Vehicle=NamedRerouteUsage()
    isMech=NamedRerouteUsage()
    isVehicle=NamedRerouteUsage()
    Pre-Skinned Local Normal=NamedRerouteUsage()
    Calcated_Bounds=NamedRerouteUsage()
    "Customization Mask (_CM B)"=NamedRerouteUsage()
   ) {materialFunction=/Game/Weapons/WeaponSkins/Wraps/MaterialFunctions/MF_Wrap_Chip}.Result_1
   Pre-Skinned Local Position=NamedRerouteUsage()
   ForwardAxis=NamedRerouteUsage()
   isState=NamedRerouteUsage()
   Pre-Skinned Local Normal=NamedRerouteUsage()
   Calcated_Bounds=NamedRerouteUsage()
   "Customization Mask (_CM B Top left)"=NamedRerouteUsage()
   "Customization Mask 2 (_CM B Top Right)"=NamedRerouteUsage()
   isCharAccent=NamedRerouteUsage()
  ) {materialFunction=/Game/Weapons/WeaponSkins/Wraps/MaterialFunctions/MF_Wrap_CosmicPulse}
  n146 = NamedRerouteUsage()
  n147 = NamedRerouteUsage()
  n148 = NamedRerouteUsage()
  n149 = NamedRerouteUsage()
  n150 = NamedRerouteUsage()
  n151 = NamedRerouteUsage()
  n152 = NamedRerouteUsage()
  n153 = MaterialFunctionCall!(
   Material=MaterialFunctionCall!(
    Material=MaterialFunctionCall!(
     Material=@n145.Result_1
     IsState=@n146
     Pre-Skinned Local Normal=@n147
     Calcated_Bounds=@n148
     Foward Axis=@n149
     "Customization Mask (_CM B Top left)"=@n150
     "Customization Mask 2 (_CM B Top Right)"=@n151
     isCharAccent=@n152
    ) {materialFunction=/Game/Weapons/WeaponSkins/Wraps/MaterialFunctions/MF_Wrap_Shapeshifter}.Result_1
    Side Project Vehicle=NamedRerouteUsage()
    isMech=NamedRerouteUsage()
    isVehicle=NamedRerouteUsage()
    Pre-Skinned Local Normal=NamedRerouteUsage()
    Calcated_Bounds=NamedRerouteUsage()
    "Customization Mask (_CM B Top left)"=NamedRerouteUsage()
    "Customization Mask 2 (_CM B Top Right)"=NamedRerouteUsage()
   ) {materialFunction=/Game/Weapons/WeaponSkins/Wraps/MaterialFunctions/MF_Wrap_Spacefighter}.Result_1
   ForwardAxis=NamedRerouteUsage()
   isState=NamedRerouteUsage()
   Pre-Skinned Local Normal=NamedRerouteUsage()
   Calcated_Bounds=NamedRerouteUsage()
   "Customization Mask (_CM B Top left)"=NamedRerouteUsage()
   "Customization Mask 2 (_CM B Top Right)"=NamedRerouteUsage()
   isCharAccent=NamedRerouteUsage()
  ) {materialFunction=/Game/Weapons/WeaponSkins/Wraps/MaterialFunctions/MF_Wrap_NeonXmasSign}
  n154 = NamedRerouteUsage()
  n155 = NamedRerouteUsage()
  n156 = NamedRerouteUsage()
  n157 = NamedRerouteUsage()
  n158 = NamedRerouteUsage()
  n159 = NamedRerouteUsage()
  n160 = NamedRerouteUsage()
  n161 = MaterialFunctionCall!(
   Material=MaterialFunctionCall!(
    Material=MaterialFunctionCall!(
     Material=@n153.Result
     isState=@n154
     Pre-Skinned Local Normal=@n155
     Calcated_Bounds=@n156
     Curavture=@n157
     Customization Mask B=@n158
     "Customization Mask (_CM B)"=@n159
     isCharAccent=@n160
    ) {materialFunction=/Game/Weapons/WeaponSkins/Wraps/MaterialFunctions/MF_WRap_ConfettiLights}.Result
    IsState=NamedRerouteUsage()
    Pre-Skinned Local Normal=NamedRerouteUsage()
    Calcated_Bounds=NamedRerouteUsage()
    Curavture=NamedRerouteUsage()
    Customization Mask B=NamedRerouteUsage()
    "Customization Mask (_CM B)"=NamedRerouteUsage()
    isCharAccent=NamedRerouteUsage()
   ) {materialFunction=/Game/Weapons/WeaponSkins/Wraps/MaterialFunctions/MF_Wrap_SmashedGlass}.Result
   IsState=NamedRerouteUsage()
   Forward Axis=NamedRerouteUsage()
   Pre-Skinned Local Normal=NamedRerouteUsage()
   Calcated_Bounds=NamedRerouteUsage()
   Customization Mask A=NamedRerouteUsage()
   Customization Mask B=NamedRerouteUsage()
   Curavture=NamedRerouteUsage()
   isCharAccent=NamedRerouteUsage()
  ) {materialFunction=/Game/Weapons/WeaponSkins/Wraps/MaterialFunctions/MF_Wrap_HunterD}
  n162 = NamedRerouteUsage()
  n163 = NamedRerouteUsage()
  n164 = NamedRerouteUsage()
  n165 = NamedRerouteUsage()
  n166 = NamedRerouteUsage()
  n167 = NamedRerouteUsage()
  n168 = NamedRerouteUsage()
  n169 = NamedRerouteUsage()
  n170 = NamedRerouteUsage()
  n171 = MaterialFunctionCall!(
   Material=MaterialFunctionCall!(
    Material=MaterialFunctionCall!(
     Material=@n161.Result
     Side Project Vehicle=@n162
     ForwardAxis=@n163
     isMech=@n164
     isVehicle=@n165
     Pre-Skinned Local Normal=@n166
     Calcated_Bounds=@n167
     Curavture=@n168
     "Customization Mask (_CM B)_1"=@n169
     "Customization Mask (_CM B)"=@n170
    ) {materialFunction=/Game/Weapons/WeaponSkins/Wraps/MaterialFunctions/MF_Wrap_StarAces}.Result
    IsState=NamedRerouteUsage()
    Pre-Skinned Local Normal=NamedRerouteUsage()
    Calcated_Bounds=NamedRerouteUsage()
    "Customization Mask (_CM B Top left)"=NamedRerouteUsage()
    "Customization Mask 2 (_CM B Top Right)"=NamedRerouteUsage()
    isCharAccent=NamedRerouteUsage()
   ) {materialFunction=/Game/Weapons/WeaponSkins/Wraps/MaterialFunctions/MF_Wrap_WarriorC}.Result
   ForwardAxis=NamedRerouteUsage()
   Is_States=NamedRerouteUsage()
   Pre-Skinned Local Normal=NamedRerouteUsage()
   Calcated_Bounds=NamedRerouteUsage()
   "Customization Mask (_CM B)"=NamedRerouteUsage()
   Curavture=NamedRerouteUsage()
   Object Bounds Center Adjustment=NamedRerouteUsage()
   Object Bounds Min Adjustment=NamedRerouteUsage()
   Object Bounds Max Adjustment=NamedRerouteUsage()
   Object-SafeZone-Virtual-Pivot=NamedRerouteUsage()
   Object-SafeZone-Virtual-CH-Switch=NamedRerouteUsage()
  ) {materialFunction=/Game/Weapons/WeaponSkins/Wraps/MaterialFunctions/MF_Wrap_TurboC}
  n172 = NamedRerouteUsage()
  n173 = NamedRerouteUsage()
  n174 = NamedRerouteUsage()
  n175 = NamedRerouteUsage()
  n176 = NamedRerouteUsage()
  n177 = NamedRerouteUsage()
  n178 = NamedRerouteUsage()
  n179 = NamedRerouteUsage()
  n180 = NamedRerouteUsage()
  n181 = NamedRerouteUsage()
  n182 = MaterialFunctionCall!(
   Material=MaterialFunctionCall!(
    Material=MaterialFunctionCall!(
     Material=@n171.Result
     ForwardAxis=@n172
     isState=@n173
     Pre-Skinned Local Normal=@n174
     Calcated_Bounds=@n175
     Curavture=@n176
     Customization Mask A=@n177
     Customization Mask B=@n178
     Object-SafeZone-Back=@n179
     Object-SafeZone-Middle=@n180
     Object-SafeZone-Front=@n181
    ) {materialFunction=/Game/Weapons/WeaponSkins/Wraps/MaterialFunctions/MF_Wrap_Mustang}.Result
    isState=NamedRerouteUsage()
    Pre-Skinned Local Normal=NamedRerouteUsage()
    Calcated_Bounds=NamedRerouteUsage()
    "Customization Mask (_CM B Top left)"=NamedRerouteUsage()
    "Customization Mask 2 (_CM B Top Right)"=NamedRerouteUsage()
    isCharAccent=NamedRerouteUsage()
   ) {materialFunction=/Game/Weapons/WeaponSkins/Wraps/MaterialFunctions/MF_Wrap_Yogurt}.Result_1
   isState=NamedRerouteUsage()
   Pre-Skinned Local Normal=NamedRerouteUsage()
   Calcated_Bounds=NamedRerouteUsage()
   "Customization Mask (_CM B Top left)"=NamedRerouteUsage()
   "Customization Mask 2 (_CM B Top Right)"=NamedRerouteUsage()
  ) {materialFunction=/Game/Weapons/WeaponSkins/Wraps/MaterialFunctions/MF_Wrap_Alakazam}
  n183 = NamedRerouteUsage()
  n184 = NamedRerouteUsage()
  n185 = NamedRerouteUsage()
  n186 = NamedRerouteUsage()
  n187 = NamedRerouteUsage()
  n188 = NamedRerouteUsage()
  n189 = NamedRerouteUsage()
  n190 = MaterialFunctionCall!(
   Material=MaterialFunctionCall!(
    Material=MaterialFunctionCall!(
     Material=@n182.Result
     isState=@n183
     Pre-Skinned Local Normal=@n184
     Calcated_Bounds=@n185
     Gmap=@n186
     "Customization Mask (_CM B Top left)"=@n187
     "Customization Mask 2 (_CM B Top Right)"=@n188
     IsStatsB=@n189
    ) {materialFunction=/Game/Weapons/WeaponSkins/Wraps/MaterialFunctions/MF_Wrap_Flora}.Result
    ForwardAxis=NamedRerouteUsage()
    isState=NamedRerouteUsage()
    Pre-Skinned Local Normal=NamedRerouteUsage()
    Calcated_Bounds=NamedRerouteUsage()
    Customization Mask A=NamedRerouteUsage()
    Curavture=NamedRerouteUsage()
    Customization Mask B=NamedRerouteUsage()
    Object Bounds Min Adjustment=NamedRerouteUsage()
    Object Bounds Max Adjustment=NamedRerouteUsage()
    Object-SafeZone-Virtual-CH-Switch=NamedRerouteUsage()
    Object-SafeZone-Virtual-Pivot=NamedRerouteUsage()
   ) {materialFunction=/Game/Weapons/WeaponSkins/Wraps/MaterialFunctions/MF_Wrap_BeeHive}.Result
   isSates=NamedRerouteUsage()
   Pre-Skinned Local Normal=NamedRerouteUsage()
   Calcated_Bounds=NamedRerouteUsage()
   Curavture=NamedRerouteUsage()
   Customization Mask B=NamedRerouteUsage()
   Customization Mask A=NamedRerouteUsage()
   isCharAccent=NamedRerouteUsage()
  ) {materialFunction=/Game/Weapons/WeaponSkins/Wraps/MaterialFunctions/MF_Wrap_AstroMap}
  n191 = NamedRerouteUsage()
  n192 = NamedRerouteUsage()
  n193 = NamedRerouteUsage()
  n194 = NamedRerouteUsage()
  n195 = NamedRerouteUsage()
  n196 = NamedRerouteUsage()
  n197 = NamedRerouteUsage()
  n198 = MaterialFunctionCall!(
   Material=MaterialFunctionCall!(
    Material=MaterialFunctionCall!(
     Material=@n190.Result
     Side Project Vehicle=@n191
     isMech=@n192
     isVehicle=@n193
     Pre-Skinned Local Normal=@n194
     Calcated_Bounds=@n195
     "Customization Mask (_CM B Top left)"=@n196
     "Customization Mask 2 (_CM B Top Right)"=@n197
    ) {materialFunction=/Game/Weapons/WeaponSkins/Wraps/MaterialFunctions/MF_Wrap_FKeepers}.Result
    isState=NamedRerouteUsage()
    Pre-Skinned Local Normal=NamedRerouteUsage()
    Calcated_Bounds=NamedRerouteUsage()
    "Customization Mask (_CM B Top left)"=NamedRerouteUsage()
    "Customization Mask 2 (_CM B Top Right)"=NamedRerouteUsage()
    IsStatsB=NamedRerouteUsage()
   ) {materialFunction=/Game/Weapons/WeaponSkins/Wraps/MaterialFunctions/MF_Wrap_StrangeCosmos}.Result
   isState=NamedRerouteUsage()
   Pre-Skinned Local Normal=NamedRerouteUsage()
   Calcated_Bounds=NamedRerouteUsage()
   Curavture=NamedRerouteUsage()
   Customization Mask A=NamedRerouteUsage()
   Customization Mask B=NamedRerouteUsage()
  ) {materialFunction=/Game/Weapons/WeaponSkins/Wraps/MaterialFunctions/MF_Wrap_GoldenSkeleton}
  n199 = NamedRerouteUsage()
  n200 = NamedRerouteUsage()
  n201 = NamedRerouteUsage()
  n202 = NamedRerouteUsage()
  n203 = NamedRerouteUsage()
  n204 = MaterialFunctionCall!(
   Material=MaterialFunctionCall!(
    Material=MaterialFunctionCall!(
     Material=@n198.Result
     IsState=@n199
     Pre-Skinned Local Normal=@n200
     Calcated_Bounds=@n201
     Curavture=@n202
     "Customization Mask (_CM B)"=@n203
    ) {materialFunction=/Game/Weapons/WeaponSkins/Wraps/MaterialFunctions/MF_Function_LanternFest}.Result_1
    IsStatsB=NamedRerouteUsage()
    IsState=NamedRerouteUsage()
    Pre-Skinned Local Normal=NamedRerouteUsage()
    Calcated_Bounds=NamedRerouteUsage()
    Customization Mask B=NamedRerouteUsage()
    Customization Mask A=NamedRerouteUsage()
    Curavture=NamedRerouteUsage()
   ) {materialFunction=/Game/Weapons/WeaponSkins/Wraps/MaterialFunctions/MF_Wrap_LavaLamp}.Result
   isState=NamedRerouteUsage()
   Pre-Skinned Local Normal=NamedRerouteUsage()
   Calcated_Bounds=NamedRerouteUsage()
   "Customization Mask (_CM B)"=NamedRerouteUsage()
  ) {materialFunction=/Game/Weapons/WeaponSkins/Wraps/MaterialFunctions/MF_Wrap_SpaceCud}
  n205 = NamedRerouteUsage()
  n206 = NamedRerouteUsage()
  n207 = NamedRerouteUsage()
  n208 = NamedRerouteUsage()
  n209 = NamedRerouteUsage()
  n210 = NamedRerouteUsage()
  n211 = MaterialFunctionCall!(
   Material=MaterialFunctionCall!(
    Material=MaterialFunctionCall!(
     Material=@n204.Result
     IsState=@n205
     Pre-Skinned Local Normal=@n206
     Calcated_Bounds=@n207
     "Customization Mask (_CM B Top left)"=@n208
     "Customization Mask 2 (_CM B Top Right)"=@n209
     isCharAccent=@n210
    ) {materialFunction=/Game/Weapons/WeaponSkins/Wraps/MaterialFunctions/MF_Wrap_RetroRainbow}.Result
    Side Project Vehicle=NamedRerouteUsage()
    isMech=NamedRerouteUsage()
    isVehicle=NamedRerouteUsage()
    Pre-Skinned Local Normal=NamedRerouteUsage()
    Calcated_Bounds=NamedRerouteUsage()
    "Customization Mask (_CM B Top left)"=NamedRerouteUsage()
    "Customization Mask 2 (_CM B Top Right)"=NamedRerouteUsage()
   ) {materialFunction=/Game/Weapons/WeaponSkins/Wraps/MaterialFunctions/MF_Wrap_VaderIn}.Result
   IsState=NamedRerouteUsage()
   Pre-Skinned Local Normal=NamedRerouteUsage()
   Calcated_Bounds=NamedRerouteUsage()
   "Customization Mask (_CM B)B"=NamedRerouteUsage()
   "Customization Mask (_CM B)"=NamedRerouteUsage()
   isCharAccent=NamedRerouteUsage()
  ) {materialFunction=/Game/Weapons/WeaponSkins/Wraps/MaterialFunctions/MF_Wrap_Faux}
  n212 = NamedRerouteUsage()
  n213 = NamedRerouteUsage()
  n214 = NamedRerouteUsage()
  n215 = NamedRerouteUsage()
  n216 = NamedRerouteUsage()
  n217 = NamedRerouteUsage()
  n218 = NamedRerouteUsage()
  n219 = NamedRerouteUsage()
  n220 = MaterialFunctionCall!(
   Material=MaterialFunctionCall!(
    Material=MaterialFunctionCall!(
     Material=@n211.Result_1
     Side Project Vehicle=@n212
     isMech=@n213
     IsState=@n214
     isVehicle=@n215
     Pre-Skinned Local Normal=@n216
     Calcated_Bounds=@n217
     "Customization Mask (_CM B Top left)"=@n218
     "Customization Mask 2 (_CM B Top Right)"=@n219
    ) {materialFunction=/Game/Weapons/WeaponSkins/Wraps/MaterialFunctions/MF_Wrap_BeamU}.Result
    IsStatsB=NamedRerouteUsage()
    Side Project Vehicle=NamedRerouteUsage()
    isMech=NamedRerouteUsage()
    isCharacter=NamedRerouteUsage()
    isVehicle=NamedRerouteUsage()
    Customization Mask B=NamedRerouteUsage()
    "Customization Mask (_CM B)"=NamedRerouteUsage()
    Object Bounds Min Adjustment=NamedRerouteUsage()
    Object Bounds Max Adjustment=NamedRerouteUsage()
    Object-SafeZone-Virtual-Pivot=NamedRerouteUsage()
    Object-SafeZone-Virtual-Pivot_CH_Switch=NamedRerouteUsage()
   ) {materialFunction=/Game/Weapons/WeaponSkins/Wraps/MaterialFunctions/MF_Wrap_CelestialGlow2}.Result
   IsStatsB=NamedRerouteUsage()
   ForwardAxis=NamedRerouteUsage()
   IsCharacter=NamedRerouteUsage()
   isMech=NamedRerouteUsage()
   isVehicle=NamedRerouteUsage()
   Side Project Vehicle=NamedRerouteUsage()
   Pre-Skinned Local Normal=NamedRerouteUsage()
   Local Position=NamedRerouteUsage()
   Object Bounds Min Adjustment=NamedRerouteUsage()
   Object Bounds Max Adjustment=NamedRerouteUsage()
   Object Bounds Center Adjustment=NamedRerouteUsage()
   "Customization Mask (_CM B)"=NamedRerouteUsage()
  ) {materialFunction=/Game/Weapons/WeaponSkins/Wraps/MaterialFunctions/MF_Cubes_Wrap}
  n221 = NamedRerouteUsage()
  n222 = NamedRerouteUsage()
  n223 = NamedRerouteUsage()
  n224 = NamedRerouteUsage()
  n225 = NamedRerouteUsage()
  n226 = MaterialFunctionCall!(
   Material=MaterialFunctionCall!(
    Material=MaterialFunctionCall!(Material=@n220.Result, Bounds=@n221, isMech=@n222, isVehicle_Side=@n223, isVehicle=@n224, "Customization Mask (_CM B)"=@n225) {materialFunction=/Game/Weapons/WeaponSkins/Wraps/MaterialFunctions/MF_Wrap_GhostHunter}.Result
    Local Normals=NamedRerouteUsage()
    Local Position=NamedRerouteUsage()
    isVehicle=NamedRerouteUsage()
    isMech=NamedRerouteUsage()
    isCharacter=NamedRerouteUsage()
    Side Project Vehicle=NamedRerouteUsage()
    Object Bounds Min Adjustment=NamedRerouteUsage()
    Object Bounds Max Adjustment=NamedRerouteUsage()
    Object Bounds Center Adjustment=NamedRerouteUsage()
    Customization Mask B=NamedRerouteUsage()
    "Customization Mask (_CM B)"=NamedRerouteUsage()
   ) {materialFunction=/Game/Weapons/WeaponSkins/Wraps/MaterialFunctions/MF_Wrap_AlDay}.Result
   IsState_B=NamedRerouteUsage()
   IsState=NamedRerouteUsage()
   Pre-Skinned Local Normal=NamedRerouteUsage()
   Calcated_Bounds=NamedRerouteUsage()
   Mask Secondary=NamedRerouteUsage()
   Mask Primary=NamedRerouteUsage()
  ) {materialFunction=/Game/Weapons/WeaponSkins/Wraps/MaterialFunctions/MF_Wrap_WLights}
  n227 = NamedRerouteUsage()
  n228 = NamedRerouteUsage()
  n229 = NamedRerouteUsage()
  n230 = NamedRerouteUsage()
  n231 = NamedRerouteUsage()
  n232 = NamedRerouteUsage()
  n233 = NamedRerouteUsage()
  n234 = NamedRerouteUsage()
  n235 = NamedRerouteUsage()
  n236 = NamedRerouteUsage()
  n237 = NamedRerouteUsage()
  n238 = NamedRerouteUsage()
  n239 = MaterialFunctionCall!(
   Material=MaterialFunctionCall!(
    Material=MaterialFunctionCall!(
     Material=@n226.Result
     IsStatsB=@n227
     ForwardAxis=@n228
     IsCharacter=@n229
     isMech=@n230
     isVehicle=@n231
     Side Project Vehicle=@n232
     Pre-Skinned Local Normal=@n233
     Local Position=@n234
     Object Bounds Min Adjustment=@n235
     Object Bounds Max Adjustment=@n236
     Object Bounds Center Adjustment=@n237
     "Customization Mask (_CM B)"=@n238
    ) {materialFunction=/Game/Weapons/WeaponSkins/Wraps/MaterialFunctions/MF_Pmint_Wrap}.Result
    IsState=NamedRerouteUsage()
    IsState_B=NamedRerouteUsage()
    Pre-Skinned Local Normal=NamedRerouteUsage()
    Calcated_Bounds=NamedRerouteUsage()
    Mask Secondary=NamedRerouteUsage()
    Mask Gmap=NamedRerouteUsage()
    Mask Primary=NamedRerouteUsage()
    Object Bounds Min Adjustment=NamedRerouteUsage()
    Object Bounds Max Adjustment=NamedRerouteUsage()
    Object-SafeZone-VP_Charcter_Preview_Switch=NamedRerouteUsage()
    Object-SafeZone-Virtual-Pivot=NamedRerouteUsage()
   ) {materialFunction=/Game/Weapons/WeaponSkins/Wraps/MaterialFunctions/MF_Wrap_SkullPunk}.Result
   IsState=NamedRerouteUsage()
   IsState_B=NamedRerouteUsage()
   Pre-Skinned Local Normal=NamedRerouteUsage()
   Calcated_Bounds=NamedRerouteUsage()
   Mask Secondary=NamedRerouteUsage()
   Mask Crevise=NamedRerouteUsage()
   Mask Primary=NamedRerouteUsage()
  ) {materialFunction=/Game/Weapons/WeaponSkins/Wraps/MaterialFunctions/MF_Wrap_SlipStream}
  n240 = NamedRerouteUsage()
  n241 = NamedRerouteUsage()
  n242 = NamedRerouteUsage()
  n243 = NamedRerouteUsage()
  n244 = NamedRerouteUsage()
  n245 = NamedRerouteUsage()
  n246 = NamedRerouteUsage()
  n247 = NamedRerouteUsage()
  n248 = NamedRerouteUsage()
  n249 = NamedRerouteUsage()
  n250 = NamedRerouteUsage()
  n251 = MaterialFunctionCall!(
   Material=MaterialFunctionCall!(
    Material=MaterialFunctionCall!(
     Material=@n239.Result
     IsState=@n240
     IsState_B=@n241
     Forward_Axis=@n242
     Pre-Skinned Local Normal=@n243
     Pre-Skinned Local Postion=@n244
     Calcated_Bounds=@n245
     Mask Crevise=@n246
     Mask Primary=@n247
     Mask Secondary=@n248
     Object-SafeZone-Virtual-Pivot=@n249
     Object-SafeZone-VP_Charcter_Preview_Switch=@n250
    ) {materialFunction=/Game/Weapons/WeaponSkins/Wraps/MaterialFunctions/MF_Wrap_HexColor}.Result_1
    Side Project Vehicle=NamedRerouteUsage()
    isMech=NamedRerouteUsage()
    isVehicle=NamedRerouteUsage()
    Pre-Skinned Local Normal=NamedRerouteUsage()
    Calcated_Bounds=NamedRerouteUsage()
    "Customization Mask (_CM B Top left)"=NamedRerouteUsage()
   ) {materialFunction=/Game/Weapons/WeaponSkins/Wraps/MaterialFunctions/MF_Wrap_EclipseS}.Result
   IsState_B=NamedRerouteUsage()
   IsState=NamedRerouteUsage()
   Pre-Skinned Local Normal=NamedRerouteUsage()
   Calcated_Bounds=NamedRerouteUsage()
   Mask Crevise=NamedRerouteUsage()
   Mask Secondary=NamedRerouteUsage()
   Mask Primary=NamedRerouteUsage()
   Object-SafeZone-Front=NamedRerouteUsage()
  ) {materialFunction=/Game/Weapons/WeaponSkins/Wraps/MaterialFunctions/MF_Wrap_WDClouds}
  n252 = NamedRerouteUsage()
  n253 = NamedRerouteUsage()
  n254 = NamedRerouteUsage()
  n255 = NamedRerouteUsage()
  n256 = NamedRerouteUsage()
  n257 = NamedRerouteUsage()
  n258 = NamedRerouteUsage()
  n259 = MaterialFunctionCall!(
   Material=MaterialFunctionCall!(
    Material=MaterialFunctionCall!(
     Material_Root=@n5
     Material=@n251.Result
     IsState_B=@n252
     IsState=@n253
     Pre-Skinned Local Normal=@n254
     Calcated_Bounds=@n255
     Mask Crevise
     Mask Secondary=@n256
     Mask Primary=@n257
     Object-SafeZone-Front=@n258
    ) {materialFunction=/Game/Weapons/WeaponSkins/Wraps/MaterialFunctions/MF_Function_TheEye}.Result
    IsState=NamedRerouteUsage()
    IsState_B=NamedRerouteUsage()
    Pre-Skinned Local Normal=NamedRerouteUsage()
    Calcated_Bounds=NamedRerouteUsage()
    Mask Crevise=NamedRerouteUsage()
    Mask Secondary=NamedRerouteUsage()
    Mask Primary=NamedRerouteUsage()
   ) {materialFunction=/Game/Weapons/WeaponSkins/Wraps/MaterialFunctions/MF_Wrap_Disarray}.Result
   IsState=NamedRerouteUsage()
   IsState_B=NamedRerouteUsage()
   Pre-Skinned Local Normal=NamedRerouteUsage()
   Calcated_Bounds=NamedRerouteUsage()
   Mask Crevise=NamedRerouteUsage()
   Mask Secondary=NamedRerouteUsage()
   Mask Primary=NamedRerouteUsage()
  ) {materialFunction=/Game/Weapons/WeaponSkins/Wraps/MaterialFunctions/MF_Wrap_Cact_Rocker}
  n260 = NamedRerouteUsage()
  n261 = NamedRerouteUsage()
  n262 = NamedRerouteUsage()
  n263 = NamedRerouteUsage()
  n264 = NamedRerouteUsage()
  n265 = NamedRerouteUsage()
  n266 = NamedRerouteUsage()
  n267 = MaterialFunctionCall!(
   Material=MaterialFunctionCall!(
    Material=MaterialFunctionCall!(
     Material=@n259.Result_1
     IsState_B=@n260
     IsState=@n261
     Pre-Skinned Local Normal=@n262
     Calcated_Bounds=@n263
     Mask Crevise=@n264
     Mask Secondary=@n265
     Mask Primary=@n266
     Object-SafeZone-Front
    ) {materialFunction=/Game/Weapons/WeaponSkins/Wraps/MaterialFunctions/MF_Wrap_EqualizerBounce}.Result
    IsState=NamedRerouteUsage()
    IsState_B=NamedRerouteUsage()
    Pre-Skinned Local Normal=NamedRerouteUsage()
    Calcated_Bounds=NamedRerouteUsage()
    Mask Crevise=NamedRerouteUsage()
    Mask Secondary=NamedRerouteUsage()
    Mask Primary=NamedRerouteUsage()
   ) {materialFunction=/Game/Weapons/WeaponSkins/Wraps/MaterialFunctions/MF_Wrap_NeonGraffitiLava}.Result
   IsState=NamedRerouteUsage()
   IsState_B=NamedRerouteUsage()
   Pre-Skinned Local Normal=NamedRerouteUsage()
   Calcated_Bounds=NamedRerouteUsage()
   Mask Crevise=NamedRerouteUsage()
   Mask Secondary=NamedRerouteUsage()
   Mask Primary=NamedRerouteUsage()
  ) {materialFunction=/Game/Weapons/WeaponSkins/Wraps/MaterialFunctions/MF_Wrap_GameBud}
  n268 = NamedRerouteUsage()
  n269 = NamedRerouteUsage()
  n270 = NamedRerouteUsage()
  n271 = NamedRerouteUsage()
  n272 = NamedRerouteUsage()
  n273 = NamedRerouteUsage()
  n274 = NamedRerouteUsage()
  n275 = MaterialFunctionCall!(
   Material=MaterialFunctionCall!(
    Material=MaterialFunctionCall!(
     Material=@n267.Result
     IsState=@n268
     IsState_B=@n269
     Pre-Skinned Local Normal=@n270
     Calcated_Bounds=@n271
     Mask Crevise=@n272
     Mask Secondary=@n273
     Mask Primary=@n274
    ) {materialFunction=/Game/Weapons/WeaponSkins/Wraps/MaterialFunctions/MF_WrapLancelot}.Result
    IsState_B=NamedRerouteUsage()
    IsState=NamedRerouteUsage()
    Pre-Skinned Local Normal=NamedRerouteUsage()
    Calcated_Bounds=NamedRerouteUsage()
    Mask Crevise=NamedRerouteUsage()
    Mask Secondary=NamedRerouteUsage()
    Mask Gmap=NamedRerouteUsage()
    Mask Primary=NamedRerouteUsage()
   ) {materialFunction=/Game/Weapons/WeaponSkins/Wraps/MaterialFunctions/MF_WrapMitos}.Result_1
   IsState=NamedRerouteUsage()
   IsState_B=NamedRerouteUsage()
   Pre-Skinned Local Normal=NamedRerouteUsage()
   Calcated_Bounds=NamedRerouteUsage()
   Mask Secondary=NamedRerouteUsage()
   Mask Crevise=NamedRerouteUsage()
   Mask Primary=NamedRerouteUsage()
  ) {materialFunction=/Game/Weapons/WeaponSkins/Wraps/MaterialFunctions/MF_Wrap_Wave}
  n276 = NamedRerouteUsage()
  n277 = NamedRerouteUsage()
  n278 = NamedRerouteUsage()
  n279 = NamedRerouteUsage()
  n280 = NamedRerouteUsage()
  n281 = NamedRerouteUsage()
  n282 = NamedRerouteUsage()
  n283 = MaterialFunctionCall!(
   Material=MaterialFunctionCall!(
    Material=MaterialFunctionCall!(
     Material=@n275.Result
     IsState=@n276
     IsState_B=@n277
     Pre-Skinned Local Normal=@n278
     Calcated_Bounds=@n279
     Mask Crevise=@n280
     Mask Secondary=@n281
     Mask Primary=@n282
    ) {materialFunction=/Game/Weapons/WeaponSkins/Wraps/MaterialFunctions/MF_Barium}.Result
    IsState=NamedRerouteUsage()
    IsState_B=NamedRerouteUsage()
    Pre-Skinned Local Normal=NamedRerouteUsage()
    Calcated_Bounds=NamedRerouteUsage()
    Mask Secondary=NamedRerouteUsage()
    Mask Crevise=NamedRerouteUsage()
    Mask Primary=NamedRerouteUsage()
   ) {materialFunction=/Game/Weapons/WeaponSkins/Wraps/MaterialFunctions/MF_Wrap_CharlotteSummer}.Result
   IsState_B=NamedRerouteUsage()
   IsState=NamedRerouteUsage()
   Pre-Skinned Local Normal=NamedRerouteUsage()
   Calcated_Bounds=NamedRerouteUsage()
   Mask Crevise=NamedRerouteUsage()
   Mask Secondary=NamedRerouteUsage()
   Mask Primary=NamedRerouteUsage()
   Object-SafeZone-VP-Character-PreView-Switch=NamedRerouteUsage()
   Object-SafeZone-Virtual-Pivot=NamedRerouteUsage()
  ) {materialFunction=/Game/Weapons/WeaponSkins/Wraps/MaterialFunctions/MF_Wrap_BadB}
  n284 = NamedRerouteUsage()
  n285 = NamedRerouteUsage()
  n286 = NamedRerouteUsage()
  n287 = NamedRerouteUsage()
  n288 = NamedRerouteUsage()
  n289 = NamedRerouteUsage()
  n290 = NamedRerouteUsage()
  n291 = NamedRerouteUsage()
  n292 = MaterialFunctionCall!(
   Material=MaterialFunctionCall!(
    Material=MaterialFunctionCall!(
     Material=@n283.Result
     IsState=@n284
     IsState_B=@n285
     Pre-Skinned Local Normal=@n286
     Calcated_Bounds=@n287
     Mask Crevise=@n288
     Mask Secondary=@n289
     Mask Primary=@n290
     ForwardAxis=@n291
    ) {materialFunction=/Game/Weapons/WeaponSkins/Wraps/MaterialFunctions/MF_Wrap_Candor}.Result
    IsState_B=NamedRerouteUsage()
    IsState=NamedRerouteUsage()
    Pre-Skinned Local Normal=NamedRerouteUsage()
    Calcated_Bounds=NamedRerouteUsage()
    Mask Crevise=NamedRerouteUsage()
    Mask Secondary=NamedRerouteUsage()
    Mask Primary=NamedRerouteUsage()
    ForwardAxis=NamedRerouteUsage()
   ) {materialFunction=/Game/Weapons/WeaponSkins/Wraps/MaterialFunctions/MF_Wrap_RenRaiderSpark}.Result
   IsState=NamedRerouteUsage()
   IsState_B=NamedRerouteUsage()
   Pre-Skinned Local Normal=NamedRerouteUsage()
   Calcated_Bounds=NamedRerouteUsage()
   Mask Crevise=NamedRerouteUsage()
   Mask Secondary=NamedRerouteUsage()
   Mask Primary=NamedRerouteUsage()
  ) {materialFunction=/Game/Weapons/WeaponSkins/Wraps/MaterialFunctions/MF_Wrap_Cons}
  n293 = NamedRerouteUsage()
  n294 = NamedRerouteUsage()
  n295 = NamedRerouteUsage()
  n296 = NamedRerouteUsage()
  n297 = NamedRerouteUsage()
  n298 = NamedRerouteUsage()
  n299 = NamedRerouteUsage()
  n300 = MaterialFunctionCall!(
   Material=MaterialFunctionCall!(
    Material=MaterialFunctionCall!(
     Material=@n292.Result
     IsState=@n293
     IsState_B=@n294
     Pre-Skinned Local Normal=@n295
     Calcated_Bounds=@n296
     Mask Crevise=@n297
     Mask Secondary=@n298
     Mask Primary=@n299
    ) {materialFunction=/Game/Weapons/WeaponSkins/Wraps/MaterialFunctions/MF_Wrap_VGlitch}.Result
    IsState=NamedRerouteUsage()
    IsState_B=NamedRerouteUsage()
    Pre-Skinned Local Normal=NamedRerouteUsage()
    Calcated_Bounds=NamedRerouteUsage()
    Mask Crevise=NamedRerouteUsage()
    Mask Secondary=NamedRerouteUsage()
    Mask Primary=NamedRerouteUsage()
   ) {materialFunction=/Game/Weapons/WeaponSkins/Wraps/MaterialFunctions/MF_Wrap_NeonBlob}.Result
   IsState_B=NamedRerouteUsage()
   IsState=NamedRerouteUsage()
   Pre-Skinned Local Normal=NamedRerouteUsage()
   Pre-Skinned Position=NamedRerouteUsage()
   Calcated_Bounds=NamedRerouteUsage()
   Mask Crevise=NamedRerouteUsage()
   Mask Primary=NamedRerouteUsage()
   Mask Secondary=NamedRerouteUsage()
   Bounds_Max=NamedRerouteUsage()
   Bounds_min=NamedRerouteUsage()
   Object-SafeZone-Front=NamedRerouteUsage()
   Object-SafeZone-Middle=NamedRerouteUsage()
   Object-SafeZone-Back=NamedRerouteUsage()
  ) {materialFunction=/Game/Weapons/WeaponSkins/Wraps/MaterialFunctions/MF_Wrap_MagicMead}
  n301 = NamedRerouteUsage()
  n302 = NamedRerouteUsage()
  n303 = NamedRerouteUsage()
  n304 = NamedRerouteUsage()
  n305 = NamedRerouteUsage()
  n306 = MaterialFunctionCall!(
   Material=MaterialFunctionCall!(
    Material=MaterialFunctionCall!(Material=@n300.Result, Primary Mask=@n301, Secondary Mask_1=@n302, In CM Mask=@n303, Mask GMAP=@n304, Mask Crevice=@n305) {materialFunction=/Game/Weapons/WeaponSkins/Wraps/MaterialFunctions/MF_Wrap_RadishMemory}.Result
    Material Root=NamedRerouteUsage()
    IsState_B=NamedRerouteUsage()
    IsState=NamedRerouteUsage()
    Pre-Skinned Local Normal=NamedRerouteUsage()
    Calcated_Bounds=NamedRerouteUsage()
    Mask Crevise=NamedRerouteUsage()
    Mask Secondary=NamedRerouteUsage()
    Mask Primary=NamedRerouteUsage()
   ) {materialFunction=/Game/Weapons/WeaponSkins/Wraps/MaterialFunctions/MF_ShinySt}.Result
   Material_Root=NamedRerouteUsage()
   IsState_B=NamedRerouteUsage()
   IsState=NamedRerouteUsage()
   Pre-Skinned Local Normal=NamedRerouteUsage()
   Calcated_Bounds=NamedRerouteUsage()
   Mask Crevise=NamedRerouteUsage()
   Mask Secondary=NamedRerouteUsage()
   Mask Primary=NamedRerouteUsage()
  ) {materialFunction=/Game/Weapons/WeaponSkins/Wraps/MaterialFunctions/MF_NitroF}
  n307 = NamedRerouteUsage()
  n308 = NamedRerouteUsage()
  n309 = NamedRerouteUsage()
  n310 = NamedRerouteUsage()
  n311 = NamedRerouteUsage()
  n312 = NamedRerouteUsage()
  n313 = NamedRerouteUsage()
  n314 = NamedRerouteUsage()
  n315 = NamedRerouteUsage()
  n316 = NamedRerouteUsage()
  NamedRerouteDeclaration(None=MaterialFunctionCall!(
    Material_1=MaterialFunctionCall!(
     Material=@n306.Result
     Material_Root=@n307
     IsState_B=@n308
     IsState=@n309
     Pre-Skinned Local Normal=@n310
     Calcated_Bounds=@n311
     Mask Crevise=@n312
     Mask Secondary=@n313
     Mask Primary=@n314
     Object-SafeZone-VP-Character-PreView-Switch=@n315
     Object-SafeZone-Virtual-Pivot=@n316
    ) {materialFunction=/Game/Weapons/WeaponSkins/Wraps/MaterialFunctions/MF_Wrap_GalaxyKn}.Result
    Material_Root=NamedRerouteUsage()
    IsState_B=NamedRerouteUsage()
    IsState=NamedRerouteUsage()
    Pre-Skinned Local Normal=NamedRerouteUsage()
    Calcated_Bounds=NamedRerouteUsage()
    Mask Crevise=NamedRerouteUsage()
    Mask Secondary=NamedRerouteUsage()
    Mask Primary=NamedRerouteUsage()
    Object-SafeZone-VP-Character-PreView-Switch=NamedRerouteUsage()
    Object-SafeZone-Virtual-Pivot=NamedRerouteUsage()
   ) {materialFunction=/Game/Weapons/WeaponSkins/Wraps/MaterialFunctions/MF_Wrap_TigerR}.Result) {name=Results, nodeColor=<rgba 1 0.9375 0 1>}
  NamedRerouteDeclaration(None=FunctionInput() {bUsePreviewValueAsDefault=true, inputName=Object Bounds Center Adjustment}) {name=Bounds_Center, nodeColor=<rgba 1 0 0.4811 1>}
  NamedRerouteDeclaration(None=FunctionInput() {bUsePreviewValueAsDefault=true, inputName=Rotation X Axis, inputType=FunctionInput_Scalar}) {name=Rotation X, nodeColor=<rgba 0.7812 0 1 1>}
  n317 = FunctionInput() {bUsePreviewValueAsDefault=true, inputName=Wrap ScaleMultiplier, inputType=FunctionInput_Scalar, previewValue=<1 0 0 1>}
  NamedRerouteDeclaration(None=@n317) {name=Scale_Multiplier, nodeColor=<rgba 1 0 0.3516 1>}
  n318 = ScalarParameter() {group=Primary Triplaner, parameterName=Weapon Z Axis Rot, sortPriority=-34}
  n319 = ScalarParameter() {group=Primary Triplaner, parameterName=Vehicle Z Axis Rot, sortPriority=-33}
  n320 = NamedRerouteUsage()
  NamedRerouteDeclaration(None=Add(
    A=FunctionInput() {bUsePreviewValueAsDefault=true, inputName=Rotation Z Axis, inputType=FunctionInput_Scalar}
    B=LinearInterpolate(A=@n318, B=@n319, Alpha=@n320)
   )) {name=Rotation Z, nodeColor=<rgba 0.7812 0 1 1>}
  n321 = MaterialFunctionCall(Float4=NamedRerouteDeclaration(None=FunctionInput() {inputName=IsStateS, inputType=FunctionInput_Vector4, sortPriority=-1}) {name=Is_States, desc="R=vehicle G=VehicleSide =R=Mech A=Character", nodeColor=<rgba 0 0.9253 1 1>}) {materialFunction=&BreakOutFloat4Components}
  NamedRerouteDeclaration(None=@n321.B) {name=isMech, nodeColor=<rgba 0 0.9253 1 1>}
  NamedRerouteDeclaration(None=@n321.G) {name=isVehicleSide, nodeColor=<rgba 0 0.9253 1 1>}
  NamedRerouteDeclaration(None=@n321.R) {name=isVehicle, nodeColor=<rgba 0 0.9253 1 1>}
  NamedRerouteDeclaration(None=@n321.A) {name=isCharacter, nodeColor=<rgba 0 0.9253 1 1>}
  NamedRerouteDeclaration(None=@n1.R) {name=IsBackpack, nodeColor=<rgba 0 0.9253 1 1>}
  NamedRerouteUsage()
  NamedRerouteUsage()
  NamedRerouteUsage()
  NamedRerouteUsage()
  NamedRerouteUsage()
  NamedRerouteUsage()
  NamedRerouteUsage()
  NamedRerouteUsage()
  NamedRerouteUsage()
  NamedRerouteUsage()
  NamedRerouteUsage()
  NamedRerouteUsage()
  NamedRerouteUsage()
  NamedRerouteUsage()
 ]}
