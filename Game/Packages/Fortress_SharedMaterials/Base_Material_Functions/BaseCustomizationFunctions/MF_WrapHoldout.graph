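# Material function that keeps cosmetic wraps off marked areas of weapons by blending the original material back over the wrapped one.
# A tight SphereMask on vertex color alpha selects areas painted at the hard-coded value of 1, scaled by the Enable Wrap Holdout parameter.
# That mask drives a BlendMaterialAttributes behind a static Enable switch, which outputs only the wrapped material when off.
# The normal dot product and pixel normal nodes are unconnected.
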MaterialFunction /Game/Packages/Fortress_SharedMaterials/Base_Material_Functions/BaseCustomizationFunctions/MF_WrapHoldout
graph {nodes [
  DotProduct(A=VertexNormalWS(), B=Constant3Vector() {constant=<rgba 0 0 1 0>})
  n1 = FunctionInput() {bUsePreviewValueAsDefault=true, inputName=Wrapped Material, inputType=FunctionInput_MaterialAttributes, sortPriority=10}
  n2 = FunctionInput() {bUsePreviewValueAsDefault=true, inputName=Original Material, inputType=FunctionInput_MaterialAttributes, sortPriority=5}
  n3 = Multiply(
   A=SphereMask(
    A=VertexColor().A
    B=Constant() {desc="Value of alpha channel needed\r\nto prevent wraps from showing \r\non weapons\r\n\r\nHard coded to not change standard", r=1}
   ) {attenuationRadius=0.01}
   B=ScalarParameter() {group=Wrap Holdout, parameterName=Enable Wrap Holdout, sliderMax=1, sortPriority=100}
  )
  n4 = StaticBool()
  FunctionOutput(None=StaticSwitch(
    True=BlendMaterialAttributes(A=@n1, B=@n2, Alpha=@n3)
    False=@n1
    Value=FunctionInput(Preview=@n4) {bUsePreviewValueAsDefault=true, inputName=Enable, inputType=FunctionInput_StaticBool, sortPriority=15}
   ))
  PixelNormalWS()
 ]}
