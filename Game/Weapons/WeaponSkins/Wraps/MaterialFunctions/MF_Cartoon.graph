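# Weapon wrap material function that adds an optional cartoon look, blended over the input material by the customization mask when the Cartoon static switch is on.
# Sun and adjustable-vector dot products, sharpened by a 50x multiply and clamp, pick between shift and shadow colors for primary and secondary regions.
# The crevasse-scaled colors mix evenly with MF_Cartoon_Spec into emissive at full roughness.
# The sun vector uses only its R component, and FortniteSunnySideFresnel is unconnected.
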MaterialFunction /Game/Weapons/WeaponSkins/Wraps/MaterialFunctions/MF_Cartoon
graph {nodes [
  n1 = FunctionInput() {inputName=Material, inputType=FunctionInput_MaterialAttributes, sortPriority=-200}
  n2 = MaterialFunctionCall() {materialFunction=&/MF_Cartoon_Spec}
  n3 = FunctionInput() {bUsePreviewValueAsDefault=true, inputName="Customization Crevasse (_CM G)", inputType=FunctionInput_Scalar, sortPriority=21}
  n4 = CollectionParameter() {
   collection=&FortniteMaterialParameters
   parameterName=SunlightDirectionalVector
  }
  n5 = MaterialFunctionCall(Float4=@n4) {materialFunction=&BreakOutFloat4Components}
  n6 = MaterialFunctionCall(X=@n5.R, Y=@n5.R, Z=@n5.R) {materialFunction=&MakeFloat3}
  n7 = Clamp(None=DotProduct(A=Normalize(VectorInput=@n6.Result), B=PixelNormalWS()))
  n8 = DotProduct(A=VectorParameter() {group=Cartoon, parameterName=vector_light_adjust, defaultValue=<rgba 0.4688 0.2058 0.2718 1>}.RGB, B=VertexNormalWS())
  n9 = Clamp(None=Multiply(A=Multiply(A=@n7, B=@n8), B=Constant() {r=50}))
  n10 = LinearInterpolate(
   A=Multiply(A=VectorParameter() {group=Cartoon, parameterName=Color_Shadow, defaultValue=<rgba 0.0398 0.0405 0.0469 1>}.RGB, B=@n3)
   B=Multiply(A=VectorParameter() {group=Cartoon, parameterName=BaseColor_shift, defaultValue=<rgba 0.3253 0.4064 0.4635 1>}.RGB, B=@n3)
   Alpha=@n9
  )
  n11 = LinearInterpolate(
   A=Multiply(A=VectorParameter() {group=Cartoon, parameterName=SecondaryColor_shift, defaultValue=<rgba 0.4635 0.0356 0.3859 1>}.RGB, B=@n3)
   B=Multiply(A=VectorParameter() {group=Cartoon, parameterName=SecondaryColor_Shadow, defaultValue=<rgba 0.0469 0.0092 0.0111 1>}.RGB, B=@n3)
   Alpha=@n9
  )
  n12 = FunctionInput() {bUsePreviewValueAsDefault=true, inputName=Secondary Color, inputType=FunctionInput_Scalar, sortPriority=24}
  n13 = MakeMaterialAttributes!(
   BaseColor
   Metallic
   Specular
   Roughness=Constant() {r=1}
   Anisotropy
   EmissiveColor=LinearInterpolate(A=Multiply(A=@n2.Result, B=@n3), B=LinearInterpolate(A=@n10, B=@n11, Alpha=@n12))
   Opacity
   OpacityMask
   Normal
   Tangent
   WorldPositionOffset
   SubsurfaceColor
   ClearCoat
   ClearCoatRoughness
   AmbientOcclusion
   Refraction
   CustomizedUVs_0
   CustomizedUVs_1
   CustomizedUVs_2
   CustomizedUVs_3
   CustomizedUVs_4
   CustomizedUVs_5
   CustomizedUVs_6
   CustomizedUVs_7
   PixelDepthOffset
   ShadingModel
   Displacement
  )
  n14 = FunctionInput() {bUsePreviewValueAsDefault=true, inputName="Customization Mask (_CM B)", inputType=FunctionInput_Scalar, sortPriority=24}
  FunctionOutput(None=StaticSwitchParameter(True=BlendMaterialAttributes(A=@n1, B=@n13, Alpha=@n14), False=@n1) {group=Customization, parameterName=Cartoon})
  MaterialFunctionCall() {materialFunction=/Game/Weapons/WeaponSkins/MaterialFunction/FortniteSunnySideFresnel}
 ]}
