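# Material function for the Columbus weapon wrap, which restyles non-metallic areas of an incoming material as a two-tone clear-coat plastic.
# The CM2 input lerps between A and B parameters for color, metalness, roughness and clear coat.
# The original Metallic masks the blend back to the source attributes, and the "Use Columbus" static switch bypasses the effect.
# CM1 and the leftover Lerp, Color Dodge and Power nodes are unused.
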MaterialFunction /Game/Weapons/WeaponSkins/Wraps/MaterialFunctions/MF_Columbus
graph {nodes [
  FunctionInput() {inputName=CM1, inputType=FunctionInput_Scalar}
  n1 = FunctionInput() {inputName=Material, inputType=FunctionInput_MaterialAttributes}
  n2 = GetMaterialAttributes(None=@n1) {attributeGetTypes [
    #1
    #2
    #3
    #4
    #5
   ]}
  n3 = FunctionInput() {inputName=CM2}
  n4 = SetMaterialAttributes!(
   MaterialAttributes=@n2.MaterialAttributes
   BaseColor=LinearInterpolate(
    A=VectorParameter() {parameterName=PlasticA, defaultValue=<rgba 0.9 0.9 0.9 1>}.RGB
    B=VectorParameter() {parameterName=PlasticB, defaultValue=<rgba 0.1146 0.1146 0.1146 1>}.RGB
    Alpha=@n3
   )
   Metallic=LinearInterpolate(A=ScalarParameter() {parameterName=Metalness_A}, B=ScalarParameter() {defaultValue=1, parameterName=Metalness_B}, Alpha=@n3)
   Specular=@n2.MaterialAttributes
   Roughness=LinearInterpolate(
    A=ScalarParameter() {defaultValue=0.2, parameterName=RoughnessA}
    B=ScalarParameter() {defaultValue=0.6, parameterName=RoughnessB}
    Alpha=@n3
   )
   EmissiveColor=@n2.MaterialAttributes
   CustomData0=LinearInterpolate(A=ScalarParameter() {defaultValue=1, parameterName=ClearCoatA}, B=ScalarParameter() {parameterName=ClearCoatB}, Alpha=@n3)
   CustomData1=ScalarParameter() {defaultValue=0.1, parameterName=asdfasdfsda}
  ) {attributeSetTypes [
    #1
    #2
    #3
    #4
    #5
    #6
    #7
   ]}
  n5 = Saturate(None=Min(A=@n2.Metallic))
  FunctionOutput(None=StaticSwitchParameter(True=BlendMaterialAttributes(A=@n4, B=@n2.MaterialAttributes, Alpha=@n5), False=@n1) {group=Customization_Wrap, parameterName=Use Columbus})
  LinearInterpolate()
  MaterialFunctionCall!(Blend, Base) {materialFunction=/Engine/Functions/Engine_MaterialFunctions03/Blends/Blend_ColorDodge}
  Power(Base=@n2.Metallic) {constExponent=4}
 ]}
