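# Wrap-skin material function that adds curvature-driven glow to weapons, vehicles and mechs behind the Curvature_Emissive switch.
# It sums curvature from the _CM green channel, inner glow and bounds-based gradients, with settings lerped by object type.
# The sum drives a color lerp, gradient texture or curve atlas, optionally masked through HeightLerp.
# The result darkens and tints emissive and base color, scales emissive by time of day, and blends via customization masks.
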MaterialFunction /Game/Weapons/WeaponSkins/Wraps/MaterialFunctions/MF_Curvture_Emissive
graph {nodes [
  n1 = FunctionInput(Preview=TextureSample() {samplerType=SAMPLERTYPE_Masks, texture=/Game/Weapons/FORT_Rifles/Textures/T_Scar_CM}.G) {bUsePreviewValueAsDefault=true, inputName="Customization Crevasse (_CM G)", inputType=FunctionInput_Scalar, sortPriority=21}
  n2 = NamedRerouteUsage()
  n3 = ChannelMaskParameter(None=@n2) {desc=Overrided To TopSideProjection, group=Emissive_Curvture, parameterName=Curvature_TPS_Mask_01_Channel, sortPriority=11}
  n4 = FunctionInput() {inputName=Triplaner 01 Mask, inputType=FunctionInput_Vector4, sortPriority=-20}
  n5 = StaticSwitch(
   True=@n3.R
   False=StaticSwitchParameter(
    True=ChannelMaskParameter(None=@n2) {desc=Overrided To TopSideProjection, group=Emissive_Curvture, parameterName=Curvature_TS_Mask_01_Channel, sortPriority=11}.R
    False=ChannelMaskParameter(None=@n4) {group=Emissive_Curvture, parameterName=Curvature_TriPlaner_01_Mask_Channel, sortPriority=11}.R
   ) {group=Customization, parameterName=TopSideProjection}
   Value=StaticBoolParameter() {group=Customization, parameterName="Custmization  V2", sortPriority=-500}
  )
  n6 = Multiply(A=Multiply(A=OneMinus(None=@n1), B=Constant() {r=2}), B=OneMinus(None=@n5))
  n7 = Multiply(A=@n1, B=Constant() {r=2})
  n8 = StaticSwitchParameter(True=If(A=@n1, B=Constant() {r=0.5}, "A > B"=OneMinus(None=@n6), "A < B"=Multiply(A=@n7, B=@n5)), False=@n1) {group=Emissive_Curvture, parameterName=Curvature_Triplaner_01_Mask_Add, sortPriority=11}
  n9 = StaticSwitchParameter(
   True=Multiply(A=@n8, B=ScalarParameter() {defaultValue=2.5, group=Emissive_Curvture_Mask, parameterName=MaskCurvature_Strength Weapon, sortPriority=204})
   False=@n8
  ) {group=Emissive_Curvture_Mask, parameterName=MaskCurvature_Invert, sortPriority=201}
  n10 = Add(A=@n9) {constB=-0.5}
  n11 = Add(A=OneMinus(None=@n9)) {constB=-0.5}
  Add(A=Multiply(A=Saturate(None=@n10)), B=Multiply(A=Saturate(None=@n11)) {constB=0})
  FunctionInput() {bUsePreviewValueAsDefault=true, inputName=IsCharacter, inputType=FunctionInput_Scalar}
  FunctionInput(Preview=Constant()) {bUsePreviewValueAsDefault=true, inputName=isVehicle, inputType=FunctionInput_Scalar, sortPriority=40}
  FunctionInput(Preview=Constant()) {bUsePreviewValueAsDefault=true, inputName=isMech, inputType=FunctionInput_Scalar, sortPriority=41}
  n12 = FunctionInput(Preview=SetMaterialAttributes!(MaterialAttributes, Normal=TextureSample() {samplerType=SAMPLERTYPE_Normal, texture=/Game/Weapons/FORT_Rifles/T_SCAR_N}.RGB) {attributeSetTypes [#1]}) {inputName=Material, inputType=FunctionInput_MaterialAttributes, sortPriority=-200}
  n13 = GetMaterialAttributes(None=@n12) {attributeGetTypes [#1, #2, #3]}
  n14 = ScalarParameter() {defaultValue=1, group=Emissive_Curvture, parameterName=Curvature_Peaks_Strength, sortPriority=14}
  n15 = Power(
   Base=Saturate(None=Add(A=@n8) {constB=-0.5})
   Exp=ScalarParameter() {defaultValue=1, group=Emissive_Curvture, parameterName=Curvature_Peaks_Expo, sortPriority=14}
  )
  n16 = OneMinus(None=@n8)
  n17 = Power(
   Base=Saturate(None=Add(A=@n16) {constB=-0.5})
   Exp=ScalarParameter() {defaultValue=1, group=Emissive_Curvture, parameterName=Curvature_Pits_Expo, sortPriority=14}
  )
  n18 = ScalarParameter() {group=Emissive_Curvture, parameterName=Curvature_Pits_Strength, sortPriority=14}
  n19 = Clamp(
   None=@n8
   Min=ScalarParameter() {defaultValue=0.3, group=Emissive_Curvture, parameterName=Curvature_Clamp_Min, sortPriority=12}
   Max=ScalarParameter() {defaultValue=0.5, group=Emissive_Curvture, parameterName=Curvature_Clamp_Max, sortPriority=13}
  ) {maxDefault=0.5}
  n20 = StaticSwitchParameter(True=OneMinus(None=@n19), False=@n19) {group=Emissive_Curvture, parameterName=Curvature_Invert, sortPriority=11}
  n21 = Multiply(
   A=Power(Base=@n20, Exp=ScalarParameter() {defaultValue=2.5, group=Emissive_Curvture, parameterName=Curvature_Expo_Weapon, sortPriority=14})
   B=ScalarParameter() {defaultValue=1, group=Emissive_Curvture, parameterName=Curvature_Strength_Weapon, sortPriority=15}
  )
  n22 = Multiply(
   A=Power(Base=@n20, Exp=ScalarParameter() {defaultValue=2.5, group=Emissive_Curvture, parameterName=Curvature_Expo_Vehicle, sortPriority=14})
   B=ScalarParameter() {defaultValue=1, group=Emissive_Curvture, parameterName=Curvature_Strength_Vehicle, sortPriority=15}
  )
  n23 = NamedRerouteUsage()
  n24 = StaticSwitchParameter(True=Add(A=Multiply(A=@n14, B=@n15), B=Multiply(A=@n17, B=@n18)), False=Saturate(None=LinearInterpolate(A=@n21, B=@n22, Alpha=@n23))) {group=Emissive_Curvture, parameterName="Curvature-Peak\\Pits", sortPriority=10}
  n25 = ScalarParameter() {defaultValue=0.75, group=Emissive_Curvture, parameterName=Curvature_Opactiy, sortPriority=10}
  n26 = ComponentMask(None=@n13.Normal) {g=true, r=true}
  n27 = NamedRerouteUsage()
  n28 = LinearInterpolate(
   A=LinearInterpolate(
    A=ScalarParameter() {defaultValue=0.2, group=Emissive_Curvture, parameterName=NormalMap_Weapon_influence, sortPriority=4}
    B=ScalarParameter() {defaultValue=0.1, group=Emissive_Curvture, parameterName=NormalMap_Vehicle_influence, sortPriority=5}
    Alpha=@n23
   )
   B=ScalarParameter() {defaultValue=0.1, group=Emissive_Curvture, parameterName=NormalMap_Mech_influence, sortPriority=5}
   Alpha=@n27
  )
  n29 = FunctionInput(Preview=VertexInterpolator(VS=PreSkinnedNormal()).PS) {bUsePreviewValueAsDefault=true, inputName=Local Normals}
  n30 = LinearInterpolate(A=Abs(None=ComponentMask(None=@n29) {b=true, r=true}), B=Abs(None=ComponentMask(None=@n29) {b=true, g=true}), Alpha=@n23)
  n31 = LinearInterpolate(
   A=ScalarParameter() {defaultValue=0.05, group=Emissive_Curvture, parameterName=SkinNormal_influence, sortPriority=6}
   B=ScalarParameter() {defaultValue=0.1, group=Emissive_Curvture, parameterName=SkinNormal_Mech_influence, sortPriority=6}
   Alpha=@n27
  )
  n32 = StaticSwitchParameter(True=Constant2Vector(), False=Add(A=Multiply(A=@n26, B=@n28), B=Multiply(A=@n30, B=@n31))) {group=Emissive_Curvture, parameterName=NormalMap_influence_disable, sortPriority=4}
  n33 = Add(A=@n32, B=NamedRerouteUsage())
  n34 = TextureSampleParameter2D(UVs=@n33) {
   group=Emissive_Curvture
   parameterName=InnerGlow Texture
   samplerType=SAMPLERTYPE_Masks
   sortPriority=23
   texture=/Game/Global/Textures/Default/Masks_and_Gradients/T_VertGradient_MASK
  }
  n35 = SphereMask(
   A=@n33
   B=Constant() {r=0.5}
   Radius=ScalarParameter() {defaultValue=0.5, group=Emissive_Curvture, parameterName=InnerGlow_Radius_A, sortPriority=30}
   Hardness=ScalarParameter() {defaultValue=-2, group=Emissive_Curvture, parameterName=InnerGlow_Hardness_A, sortPriority=31}
  )
  n36 = StaticSwitchParameter(True=OneMinus(None=@n35), False=@n35) {group=Emissive_Curvture, parameterName=InnerGlow_Invert, sortPriority=23}
  n37 = Add(A=ComponentMask(None=@n13.Normal) {r=true}, B=ComponentMask(None=@n13.Normal) {g=true})
  n38 = PreSkinnedLocalBounds()
  n39 = FunctionInput(Preview=Constant3Vector() {constant=<rgba 0 31.5 0 1>}) {inputName=Object Bounds Min Adjustment, sortPriority=51}
  n40 = Add(A=@n38.Min, B=@n39)
  n41 = FunctionInput(Preview=VertexInterpolator(VS=PreSkinnedPosition()).PS) {bUsePreviewValueAsDefault=true, inputName=Local Position}
  n42 = Subtract(A=@n41, B=@n40)
  n43 = FunctionInput(Preview=Constant3Vector() {constant=<rgba 0 -0.5 0 1>}) {inputName=Object Bounds Max Adjustment, sortPriority=52}
  n44 = Subtract(A=Add(A=@n38.Max, B=@n43), B=@n40)
  n45 = Multiply(A=@n44) {constB=-0.5}
  n46 = FunctionInput(Preview=Constant3Vector() {constant=<rgba 0.2122 0 -0.09 0>}) {inputName=Object Bounds Center Adjustment, sortPriority=50}
  n47 = MaterialFunctionCall(Color=@n44) {materialFunction=&SplitComponents}
  n48 = Max(A=Max(A=@n47.R, B=@n47.R), B=@n47.B)
  n49 = SphereMask(
   A=Add(A=Add(A=@n42, B=@n45), B=Multiply(A=@n46, B=@n48))
   B=Constant3Vector() {constant=<rgba 0.5 0.5 0.5 0>}
   Radius=Multiply(A=@n48, B=ScalarParameter() {defaultValue=0.5, group=Emissive_Curvture, parameterName=InnerGlow_Radius, sortPriority=30})
   Hardness=ScalarParameter() {defaultValue=-2, group=Emissive_Curvture, parameterName=InnerGlow_Hardness, sortPriority=31}
  )
  n50 = Multiply(
   A=OneMinus(None=Multiply(A=@n37, B=@n28))
   B=StaticSwitchParameter(True=OneMinus(None=@n49), False=@n49) {group=Emissive_Curvture, parameterName=InnerGlow_Invert, sortPriority=23}
  )
  n51 = Multiply(
   A=StaticSwitchParameter(
    True=ChannelMaskParameter(None=@n34.RGB) {group=Emissive_Curvture, parameterName=InnerGlow Texture_Channel, sortPriority=22}.R
    False=StaticSwitchParameter(True=@n50, False=@n36) {group=Emissive_Curvture, parameterName=InnerGlow_Cheap, sortPriority=23}
   ) {group=Emissive_Curvture, parameterName=InnerGlow-Texture--------------------------------, sortPriority=21}
   B=ScalarParameter() {defaultValue=0.8, group=Emissive_Curvture, parameterName=InnerGlow Opacity, sortPriority=20}
  )
  n52 = Constant()
  n53 = Constant()
  n54 = PreSkinnedLocalBounds()
  n55 = MaterialFunctionCall(Color=@n54.Min) {materialFunction=&SplitComponents}
  n56 = Constant()
  n57 = LinearInterpolate(A=@n54.Min, B=AppendVector(A=AppendVector(A=@n55.R, B=@n56), B=@n55.B), Alpha=@n23)
  n58 = Divide(A=Subtract(A=@n41, B=LinearInterpolate(A=@n57, B=@n54.Min, Alpha=@n27)), B=Subtract(A=@n54.Max, B=@n54.Max))
  n59 = MaterialFunctionCall!(
   is Vehicle=NamedRerouteUsage()
   Object Bounds Max Adjustment=@n43
   Object Bounds Min Adjustment=@n39
   Object Bounds Center Adjustment=@n46
   isMech=@n27
   Side Project Vehicle
   Local Position=@n41
  ) {materialFunction=&/MF_WrapBounds}
  n60 = VectorParameter() {group=Emissive_Curvture, parameterName=Gradent_WeaponOffset_ CE, sortPriority=110, defaultValue=<rgba 0 -1 0 1>}
  n61 = VectorParameter() {group=Emissive_Curvture, parameterName=Gradent_VehicleOffset_ CE, sortPriority=112, defaultValue=<rgba -1 0 0 1>}
  n62 = LinearInterpolate(A=@n60.RGB, B=@n61.RGB, Alpha=@n23)
  n63 = VectorParameter() {group=Emissive_Curvture, parameterName=Gradent_MechOffset_ CE, sortPriority=114, defaultValue=<rgba 0 0 -1 1>}
  n64 = LinearInterpolate(
   A=ScalarParameter() {group=Emissive_Curvture, parameterName=Gradent_WeaponOffset CE, sortPriority=110}
   B=ScalarParameter() {group=Emissive_Curvture, parameterName=Gradent_VehicleOffset CE, sortPriority=112}
   Alpha=@n23
  )
  n65 = ScalarParameter() {group=Emissive_Curvture, parameterName=Gradent_MechOffset CE, sortPriority=114}
  n66 = StaticBoolParameter() {group=Emissive_Curvture, parameterName=Gradent-Texture CE--------------------------------, sortPriority=101}
  n67 = Add(
   A=Add(
    A=AppendVector(A=@n32, B=@n53)
    B=StaticSwitchParameter(True=@n58, False=@n59.Result) {group=Emissive_Curvture, parameterName=Curvature_Disable_Bounds_PerObject, sortPriority=-101}
   )
   B=StaticSwitch(True=LinearInterpolate(A=@n62, B=@n63.RGB, Alpha=@n27), False=LinearInterpolate(A=@n64, B=@n65, Alpha=@n27), Value=@n66)
  )
  n68 = MaterialFunctionCall(Float3=@n67) {materialFunction=&BreakOutFloat3Components}
  n69 = FunctionInput(Preview=Constant()) {bUsePreviewValueAsDefault=true, inputName=Side Project Vehicle, inputType=FunctionInput_Scalar}
  n70 = LinearInterpolate(
   A=AppendVector(A=@n68.G, B=@n68.G)
   B=LinearInterpolate(A=AppendVector(A=@n68.R, B=@n68.R), B=AppendVector(A=@n68.R, B=@n68.R), Alpha=@n69)
   Alpha=@n23
  )
  n71 = AppendVector(A=@n68.B, B=@n68.B)
  n72 = VectorParameter() {group=Emissive_Curvture, parameterName=Gradent_Spread_XY CE, sortPriority=104, defaultValue=<rgba 1 1 1 1>}
  n73 = ComponentMask(None=@n72.RGB) {g=true, r=true}
  n74 = VectorParameter() {group=Emissive_Curvture, parameterName=Gradent_Spread_Mech_XY CE, sortPriority=104, defaultValue=<rgba 1 1 1 1>}
  n75 = ComponentMask(None=@n74.RGB) {g=true, r=true}
  n76 = VectorParameter() {group=Emissive_Curvture, parameterName=Gradent_Animated_XY CE, sortPriority=105, defaultValue=<rgba 0 0 0 1>}
  n77 = Add(
   A=Multiply(A=LinearInterpolate(A=@n70, B=@n71, Alpha=@n27), B=LinearInterpolate(A=@n73, B=@n75, Alpha=@n27))
   B=Multiply(A=Time(), B=ComponentMask(None=@n76.RGB) {g=true, r=true})
  )
  n78 = Multiply(
   A=ChannelMaskParameter(None=@n4) {group=Emissive_Curvture, parameterName=Gradent_WarpUVs_TriPlaner_01_Mask_Channel, sortPriority=107}.R
   B=ScalarParameter() {group=Emissive_Curvture, parameterName=Gradent_WarpUVs_Amount CE, sortPriority=107}
  )
  n79 = TextureSampleParameter2D(UVs=StaticSwitch(
    True=Add(A=@n77, B=@n78)
    False=@n77
    Value=StaticBoolParameter() {group=Emissive_Curvture, parameterName=Gradent_Warp_Uvs, sortPriority=107}
   )) {
   group=Emissive_Curvture
   parameterName=Gradent_Texture CE
   samplerType=SAMPLERTYPE_Masks
   sortPriority=103
   texture=&T_VertGradient_MASK
  }
  n80 = ChannelMaskParameter(None=@n79.RGB) {group=Emissive_Curvture, parameterName=Gradent_Texture_Channel CE, sortPriority=102}
  n81 = ScalarParameter() {defaultValue=1.2, group=Emissive_Curvture, parameterName=Gradent_WeaponFalloff CE, sortPriority=111}
  n82 = ScalarParameter() {defaultValue=1.2, group=Emissive_Curvture, parameterName=Gradent_VehicleFalloff CE, sortPriority=113}
  n83 = Divide(
   A=@n67
   B=LinearInterpolate(
    A=LinearInterpolate(A=@n81, B=@n82, Alpha=@n23)
    B=ScalarParameter() {defaultValue=1.2, group=Emissive_Curvture, parameterName=Gradent_MechFalloff CE, sortPriority=115}
    Alpha=@n27
   )
  )
  n84 = FunctionInput(Preview=Constant3Vector() {constant=<rgba 1 0 0 0>}) {bUsePreviewValueAsDefault=true, inputName=ForwardAxis}
  n85 = MaterialFunctionCall(Float3=Multiply(A=Saturate(None=@n83), B=@n84)) {materialFunction=&BreakOutFloat3Components}
  n86 = Max(A=Max(A=@n85.R, B=@n85.R), B=@n85.B)
  n87 = Saturate(None=StaticSwitch(
    True=@n86
    False=OneMinus(None=@n86)
    Value=StaticBoolParameter() {group=Emissive_Curvture, parameterName=Gradent_Invert CE, sortPriority=106}
   ))
  n88 = StaticSwitchParameter(
   True=Multiply(
    A=StaticSwitch(True=@n80.R, False=@n87, Value=@n66)
    B=ScalarParameter() {defaultValue=0.4, group=Emissive_Curvture, parameterName=Gradent_Opacity CE, sortPriority=100}
   )
   False=Constant()
  ) {group=Emissive_Curvture, parameterName=Gradent-Enable CE--------------------------------, sortPriority=99}
  n89 = Add(
   A=StaticSwitchParameter(True=Multiply(A=@n24, B=@n25), False=Constant()) {group=Emissive_Curvture, parameterName=Curvature-Enable--------------------------------, sortPriority=9}
   B=Add(
    A=StaticSwitchParameter(True=@n51, False=@n52) {group=Emissive_Curvture, parameterName=InnerGlow-Enable--------------------------------, sortPriority=19}
    B=@n88
   ) {constB=0}
  ) {constB=0}
  n90 = OneMinus(None=Multiply(A=@n89, B=ScalarParameter() {defaultValue=1, group=Emissive_Curvture, parameterName=Darken_Emissive, sortPriority=-96}))
  n91 = Clamp(None=@n89, Min=ScalarParameter() {defaultValue=0.01, group=Emissive_Curvture, parameterName=Glow_Clamp Min, sortPriority=8}) {maxDefault=0.99, minDefault=0.01}
  n92 = Add(A=@n91, B=Multiply(A=Time(), B=ScalarParameter() {group=Emissive_Curvture, parameterName=Glow-Gradient-Animated_Speed, sortPriority=8}))
  n93 = Frac(None=@n92)
  n94 = CurveAtlasRowParameter(CurveTime=@n93) {
   atlas=/Game/Weapons/WeaponSkins/Wraps/Textures/CruveAtlas_Wraps_01
   curve=/Game/Weapons/WeaponSkins/Wraps/Textures/C_GradSimple
   group=Emissive_Curvture
   parameterName=Glow-Gradient-Curve
   sortPriority=8
  }
  n95 = MaterialFunctionCall!(
   Index=ScalarParameter() {defaultValue=1, group=Emissive_Curvture, parameterName="Glow-Gradient-Curve  Index", sortPriority=8}
   CCinput=@n94
   CCinput_Alpha=@n94
   UV=@n93
  ) {materialFunction=&/MF_Global_CC_Tex}
  n96 = ScalarParameter() {defaultValue=1, group=Emissive_Curvture, parameterName=Glow_Intensty_Texture_Alpha, sortPriority=8}
  n97 = Multiply(A=@n96, B=@n95.Alpha)
  n98 = TextureSampleParameter2D(UVs=@n92) {
   group=Emissive_Curvture
   parameterName=Glow-Gradient-Texture
   sortPriority=8
   texture=/Game/Weapons/WeaponSkins/Wraps/Textures/FragmentedGlow/T_Wrap_FragmentedGlow_grad3
  }
  n99 = Multiply(A=@n98.A, B=@n96)
  n100 = StaticSwitchParameter(
   True=StaticSwitchParameter(True=Multiply(A=@n95.3V, B=@n97), False=Multiply(A=@n98.RGB, B=@n99)) {group=Emissive_Curvture, parameterName="Glow-Gradient-Texture--->Cruve", sortPriority=7}
   False=LinearInterpolate(
    A=VectorParameter() {group=Emissive_Curvture, parameterName=Glow_ColorB, sortPriority=8, defaultValue=<rgba 0 0 0 1>}.RGB
    B=VectorParameter() {group=Emissive_Curvture, parameterName=Glow_ColorA, sortPriority=8, defaultValue=<rgba 1 0 0 1>}.RGB
    Alpha=@n91
   )
  ) {group=Emissive_Curvture, parameterName=Glow-Gradient-Texture-------------------------------, sortPriority=7}
  n101 = VectorParameter() {group=Emissive_Curvture_Mask, parameterName=Mask_WeaponOffset_ CE, sortPriority=161, defaultValue=<rgba 0 -1 0 1>}
  n102 = StaticBoolParameter() {group=Emissive_Curvture_Mask, parameterName=Mask-Gradent-texture CE--------------------------------, sortPriority=151}
  n103 = VectorParameter() {group=Emissive_Curvture_Mask, parameterName=Mask_VehicleOffset_ CE, sortPriority=163, defaultValue=<rgba -1 0 0 1>}
  n104 = VectorParameter() {group=Emissive_Curvture_Mask, parameterName=Mask_MechOffset_ CE, sortPriority=165, defaultValue=<rgba 0 0 -1 1>}
  n105 = VectorParameter() {group=Emissive_Curvture_Mask, parameterName=Mask_UV_Spread_XY CE, sortPriority=159, defaultValue=<rgba 1 1 1 0>}
  n106 = VectorParameter() {group=Emissive_Curvture_Mask, parameterName=Mask_UV_Spread_Mech_XY CE, sortPriority=159, defaultValue=<rgba 1 1 1 0>}
  n107 = ChannelMaskParameter(None=@n4) {desc=Overrided To TopSideProjection, group=Emissive_Curvture_Mask, parameterName=Mask_WarpUVs_TS_Mask_01_Channel, sortPriority=157}
  n108 = StaticSwitchParameter(
   True=@n107.R
   False=ChannelMaskParameter(None=@n4) {group=Emissive_Curvture_Mask, parameterName=Mask_WarpUVs_TriPlaner_01_Mask_Channel, sortPriority=157}.R
  ) {group=Customization, parameterName=TopSideProjection}
  n109 = ScalarParameter() {group=Emissive_Curvture_Mask, parameterName=Mask_WarpUVs_Amount CE, sortPriority=158}
  n110 = MaterialFunctionCall!(
   ForwardAxis=@n84
   Bounds=StaticSwitchParameter(True=@n58, False=@n59.Result) {group=Emissive_Curvture_Mask, parameterName=Mask_Disable_Bounds_PerObject, sortPriority=2}
   Pre-Skinned Local Position
   isVehicle=@n23
   isMech=@n27
   Side Project Vehicle=@n69
   WeaponOffset=StaticSwitch(True=@n101.RGB, False=ScalarParameter() {group=Emissive_Curvture_Mask, parameterName=Mask_WeaponOffset CE, sortPriority=161}, Value=@n102)
   WeaponFalloff=ScalarParameter() {defaultValue=0.5, group=Emissive_Curvture_Mask, parameterName=Mask_WeaponFalloff CE, sortPriority=162}
   VehicleOffset=StaticSwitch(True=@n103.RGB, False=ScalarParameter() {group=Emissive_Curvture_Mask, parameterName=Mask_VehicleOffset CE, sortPriority=163}, Value=@n102)
   VehicleFalloff=ScalarParameter() {defaultValue=0.5, group=Emissive_Curvture_Mask, parameterName=Mask_VehicleFalloff CE, sortPriority=164}
   MechOffset=StaticSwitch(True=@n104.RGB, False=ScalarParameter() {group=Emissive_Curvture_Mask, parameterName=Mask_MechOffset CE, sortPriority=165}, Value=@n102)
   MechFalloff=ScalarParameter() {group=Emissive_Curvture_Mask, parameterName=Mask_MechFalloff CE, sortPriority=166}
   Invert_Gradent=StaticBoolParameter() {group=Emissive_Curvture_Mask, parameterName=Mask_Invert_Gradent CE, sortPriority=152}
   UV_Spread_XY=LinearInterpolate(A=@n105.RGB, B=@n106.RGB, Alpha=@n27)
   UV_Animated_XY=VectorParameter() {group=Emissive_Curvture_Mask, parameterName=Mask_UV_Animated_XY CE, sortPriority=160}.RGB
   Warp_UVs=StaticSwitchParameter(True=Multiply(A=@n108, B=@n109), False=Constant()) {group=Emissive_Curvture_Mask, parameterName=Mask_WarpUVs CE, sortPriority=156}
   Rotate_Mech
   Rotate_Vehicle_Side
   Rotate_Vehicle
  ) {materialFunction=&/MF_Wrap_Gradent_Creation}
  n111 = TextureSampleParameter2D(UVs=@n110.Uvs) {
   group=Emissive_Curvture_Mask
   parameterName=Mask_Gradent_Texture CE
   samplerType=SAMPLERTYPE_Masks
   sortPriority=155
   texture=&T_VertGradient_MASK
  }
  n112 = StaticSwitch(
   True=ChannelMaskParameter(None=@n111.RGB) {group=Emissive_Curvture_Mask, parameterName=Mask_Gradent_Texture_Channel CE, sortPriority=154}.R
   False=@n110.Gradent
   Value=@n102
  )
  n113 = ScalarParameter() {defaultValue=1, group=Emissive_Curvture_Mask, parameterName=Mask_Gradent_Clamp Max, sortPriority=151}
  n114 = Constant()
  n115 = StaticSwitchParameter(
   True=Multiply(
    A=Clamp(None=@n112, Max=@n113) {maxDefault=0.99, minDefault=0.01}
    B=ScalarParameter() {defaultValue=1, group=Emissive_Curvture_Mask, parameterName=Mask_Gradent_Opacity_CE, sortPriority=149}
   )
   False=@n114
  ) {group=Emissive_Curvture_Mask, parameterName=Mask_Gradent, sortPriority=149}
  n116 = LinearInterpolate(
   A=ScalarParameter() {defaultValue=-1, group=Emissive_Curvture_Mask, parameterName=MaskCurvature_WeaponDarks, sortPriority=204}
   B=ScalarParameter() {defaultValue=1, group=Emissive_Curvture_Mask, parameterName=MaskCurvature_WeaponLights, sortPriority=204}
   Alpha=@n9
  )
  n117 = LinearInterpolate(
   A=ScalarParameter() {defaultValue=-1, group=Emissive_Curvture_Mask, parameterName=MaskCurvature_VehicleDarks, sortPriority=205}
   B=ScalarParameter() {defaultValue=1, group=Emissive_Curvture_Mask, parameterName=MaskCurvature_VehicleLights, sortPriority=205}
   Alpha=@n9
  )
  n118 = ScalarParameter() {defaultValue=-1, group=Emissive_Curvture_Mask, parameterName=MaskCurvature_MechDark, sortPriority=206}
  n119 = ScalarParameter() {defaultValue=1, group=Emissive_Curvture_Mask, parameterName=MaskCurvature_MechLights, sortPriority=206}
  n120 = Saturate(None=LinearInterpolate(A=LinearInterpolate(A=@n116, B=@n117, Alpha=@n23), B=LinearInterpolate(A=@n118, B=@n119, Alpha=@n9), Alpha=@n27))
  n121 = ScalarParameter() {defaultValue=1, group=Emissive_Curvture_Mask, parameterName=Mask-Curvature-Opacity--------------------------------, sortPriority=200}
  n122 = Multiply(
   A=StaticSwitchParameter(True=OneMinus(None=@n120), False=@n120) {group=Emissive_Curvture_Mask, parameterName=MaskCurvature_Invert, sortPriority=201}
   B=@n121
  )
  n123 = Add(A=@n115, B=StaticSwitchParameter(True=@n122, False=@n114) {group=Emissive_Curvture_Mask, parameterName=Mask-Curvature, sortPriority=200})
  n124 = StaticBoolParameter() {group=Emissive_Curvture_Mask, parameterName=Curvature-Mask, sortPriority=142}
  n125 = NamedRerouteDeclaration(None=FunctionInput() {inputName=Topside 01 Msk, inputType=FunctionInput_Vector4, sortPriority=-20}) {name=TPSide 01_Msk, nodeColor=<rgba 0.9688 0 1 1>}
  n126 = FunctionInput() {inputName=Topside 02 Msk, inputType=FunctionInput_Vector4, sortPriority=-20}
  n127 = ChannelMaskParameter(None=@n4) {group=Emissive_Curvture_Mask, parameterName=TriPlaner 01_Mask_Channel, sortPriority=144}
  n128 = StaticSwitchParameter(
   True=FunctionInput() {inputName=Triplaner 02 Alpha, inputType=FunctionInput_Scalar, sortPriority=-20}
   False=FunctionInput() {inputName=Triplaner 01 Alpha, inputType=FunctionInput_Scalar, sortPriority=-20}
  ) {group=Emissive_Curvture_Mask, parameterName="TriPlaner 01_Alpha or TriPlaner 02_Alpha ", sortPriority=141}
  n129 = StaticSwitch(
   True=StaticSwitchParameter(
    True=ChannelMaskParameter(None=@n125) {group=Emissive_Curvture_Mask, parameterName=TopS_Mask_01_Channel, sortPriority=144}.R
    False=ChannelMaskParameter(None=@n126) {group=Emissive_Curvture_Mask, parameterName=TopS_Mas_02_Channel, sortPriority=144}.R
   ) {group=Customization, parameterName="TopSide_01_Msk ---> Topside_02_Msk"}
   False=StaticSwitchParameter(
    True=ChannelMaskParameter(None=@n4) {desc=Overrided To TopSideProjection, group=Emissive_Curvture_Mask, parameterName=TS_Mask_01_Channel, sortPriority=144}.R
    False=StaticSwitchParameter(True=@n127.R, False=@n128) {group=Emissive_Curvture_Mask, parameterName=TriPlaner 01_Mask, sortPriority=142}
   ) {group=Customization, parameterName=TopSideProjection}
   Value=StaticBoolParameter() {group=Customization, parameterName="Custmization  V2", sortPriority=-500}
  )
  n130 = StaticSwitchParameter(True=OneMinus(None=@n129), False=@n129) {group=Emissive_Curvture_Mask, parameterName=Mask_invert, sortPriority=4}
  n131 = MaterialFunctionCall(
   Transition Phase=Saturate(None=StaticSwitch(True=@n115, False=@n123, Value=@n124))
   Height Texture=StaticSwitch(True=@n122, False=@n130, Value=@n124)
   Contrast=ScalarParameter() {defaultValue=1, group=Emissive_Curvture_Mask, parameterName=Mask_HeightLerp_Contrast, sortPriority=145}
  ) {materialFunction=&HeightLerp}
  n132 = Saturate(None=StaticSwitchParameter(
    True=@n122
    False=StaticSwitchParameter(True=@n131.Alpha, False=@n130) {group=Emissive_Curvture_Mask, parameterName=Mask_HeightLerp, sortPriority=144}
   ) {group=Emissive_Curvture_Mask, parameterName=Mask-Curvature-Only, sortPriority=3})
  n133 = StaticBoolParameter() {group=Emissive_Curvture_Mask, parameterName=Mask-Glow------------------------}
  n134 = StaticSwitch(True=Multiply(A=@n100, B=@n132), False=@n100, Value=@n133)
  n135 = Add(
   A=Multiply(A=@n13.EmissiveColor, B=Saturate(None=@n90))
   B=Multiply(A=ScalarParameter() {defaultValue=1, group=Emissive_Curvture, parameterName=Glow_Emissive_Opacty, sliderMax=1, sortPriority=-95}, B=@n134)
  )
  n136 = ScalarParameter() {defaultValue=25, group=Emissive_Curvture, parameterName=Glow-Emissive-MaxClamp-Value-----------, sliderMax=1, sortPriority=-95}
  n137 = ScalarParameter() {defaultValue=1, group=Emissive_Curvture, parameterName=CE_TOD_EmissiveNight}
  n138 = ScalarParameter() {defaultValue=1, group=Emissive_Curvture, parameterName=CE_TOD_EmissiveDay}
  n139 = ScalarParameter() {defaultValue=1, group=Emissive_Curvture, parameterName=CE_TOD_EmissiveMorningEvening}
  n140 = ScalarParameter() {group=Emissive_Curvture, parameterName=Darken_BaseColor, sortPriority=-96}
  n141 = Saturate(None=Multiply(A=ScalarParameter() {group=Emissive_Curvture, parameterName=Darken_BaseColor_Use_Calc, sortPriority=-96}, B=@n89))
  n142 = StaticSwitch(True=LinearInterpolate(B=@n140, Alpha=@n132) {constA=1}, False=Multiply(A=@n140, B=OneMinus(None=@n141)), Value=@n133)
  n143 = ScalarParameter() {group=Emissive_Curvture, parameterName=Glow_BaseColor_Opacty, sliderMax=1, sortPriority=-95}
  n144 = SetMaterialAttributes!(
   MaterialAttributes=@n13.MaterialAttributes
   EmissiveColor=Multiply(
    A=Clamp(None=@n135, Max=@n136)
    B=MaterialFunctionCall(Value_Morning=@n137, Value_Day=@n138, Value_Evening=@n139, Value_Night=@n137) {materialFunction=&Scalar_Time_of_Day}.Result
   )
   BaseColor=Add(A=Multiply(A=@n13.BaseColor, B=@n142), B=Multiply(A=@n143, B=@n134))
  ) {attributeSetTypes [#2, #3]}
  n145 = GetMaterialAttributes(None=@n12)
  n146 = FunctionInput() {bUsePreviewValueAsDefault=true, inputName="Customization Mask 2 (_CM B Top Right)", inputType=FunctionInput_Scalar, sortPriority=25}
  n147 = FunctionInput() {bUsePreviewValueAsDefault=true, inputName="Customization Mask (_CM B Top left)", inputType=FunctionInput_Scalar, sortPriority=24}
  n148 = OneMinus(None=@n147)
  n149 = Saturate(None=Add(A=@n146, B=@n148))
  n150 = StaticSwitchParameter(
   True=StaticSwitchParameter(
    True=OneMinus(None=@n146)
    False=StaticSwitchParameter(True=@n146, False=@n149) {group=Emissive_Curvture, parameterName=Curvature_Primary_Mask_Exclude, sortPriority=-97}
   ) {group=Emissive_Curvture, parameterName=Curvature_Invert_Secondary_Mask, sortPriority=-98}
   False=StaticSwitchParameter(True=@n148, False=Constant()) {group=Emissive_Curvture, parameterName=Curvature_Primary_Mask, sortPriority=-99}
  ) {group=Emissive_Curvture, parameterName=Curvature_Secondary_Mask, sortPriority=-99}
  FunctionOutput(None=StaticSwitchParameter(True=BlendMaterialAttributes(A=@n144, B=@n145.MaterialAttributes, Alpha=@n150), False=@n145.MaterialAttributes) {group=Customization, parameterName=Curvature_Emissive})
  LinearInterpolate(Alpha=@n9) {constA=-1}
  LinearInterpolate(
   A=LinearInterpolate(A=Constant3Vector() {constant=<rgba 0 1 0 0>}, B=Constant3Vector() {constant=<rgba 1 0 0 0>}, Alpha=@n23)
   B=Constant3Vector() {constant=<rgba 0 0 1 0>}
   Alpha=@n27
  )
  LinearInterpolate(A=LinearInterpolate(B=Constant3Vector() {constant=<rgba 0 0 0 1>}, Alpha=@n23) {constB=0}, B=@n46, Alpha=@n27) {constB=0}
  Multiply(A=@n88)
  n151 = MaterialFunctionCall(Float4=NamedRerouteDeclaration(None=FunctionInput() {bUsePreviewValueAsDefault=true, inputName=IsState, inputType=FunctionInput_Vector4, sortPriority=4}) {name=IsState, nodeColor=<rgba 0.1497 1 0.1633 1>}) {materialFunction=&BreakOutFloat4Components}
  NamedRerouteDeclaration(None=@n151.A) {name=IsCharacter, nodeColor=<rgba 1 0.0234 0 1>}
  n152 = FunctionInput() {bUsePreviewValueAsDefault=true, inputName=IsStatsB, inputType=FunctionInput_Vector2}
  NamedRerouteDeclaration(None=ComponentMask(None=@n152) {r=true}) {name=IsBackpack, nodeColor=<rgba 0 0.9062 1 1>}
  NamedRerouteDeclaration(None=ComponentMask(None=@n152) {g=true}) {name=IsPickaxe, nodeColor=<rgba 1 0 0.6094 1>}
  n153 = FunctionInput() {bUsePreviewValueAsDefault=true, inputName=Object-SafeZone-Virtual-Pivot, inputType=FunctionInput_Vector4, sortPriority=105}
  n154 = FunctionInput() {bUsePreviewValueAsDefault=true, inputName=Object-SafeZone-VP_Charcter_Preview_Switch, inputType=FunctionInput_Scalar, sortPriority=105}
  NamedRerouteDeclaration(None=MaterialFunctionCall!(
    Camera _pos
    Axis
    Scale
    Scale_Multplyer
    Object-SafeZone-Virtual-Pivot=@n153
    Object-SafeZone-Virtual-CH-Switch=@n154
    SinglePlane Normal Axis
    SinglePlane X Axis
    PlaneCenter_distortion
   ) {materialFunction=&/MF_SafeZone_VirtualPlane}.Uvs) {name=VP_Uvs, nodeColor=<rgba 0.3594 0 1 1>}
  NamedRerouteDeclaration(None=@n1) {name=Msk_Curvature, nodeColor=<rgba 0.0781 0 1 1>}
  NamedRerouteDeclaration(None=@n151.R) {name=isVehicle, nodeColor=<rgba 1 0 0.3047 1>}
  NamedRerouteDeclaration(None=@n152) {name=IsState_B, nodeColor=<rgba 0.1497 1 0.1633 1>}
  NamedRerouteDeclaration(None=@n151.G) {name=isVehicleSide, nodeColor=<rgba 1 0 0.3047 1>}
  NamedRerouteDeclaration(None=@n151.B) {name=isMech, nodeColor=<rgba 1 0 0.3047 1>}
  NamedRerouteUsage()
 ]}
