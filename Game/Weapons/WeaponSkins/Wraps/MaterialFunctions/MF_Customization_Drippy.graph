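# MF_Customization_Drippy is a weapon-wrap material function that layers an animated drip effect over incoming material attributes, tuned separately for weapons, vehicles and mechs.
# A triplanar projection of drip diffuse and normal textures pans over pre-skinned local position over time, offset by cloud noise.
# A bounds-based directional gradient drives a HeightLerp blend and gradient-faded, view-angle-weighted emissive.
# Static switches toggle the whole effect, diffuse-as-emissive and a Z color gradient.
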MaterialFunction /Game/Weapons/WeaponSkins/Wraps/MaterialFunctions/MF_Customization_Drippy
graph {nodes [
  n1 = FunctionInput(Preview=VertexInterpolator(VS=PreSkinnedPosition()).PS) {bUsePreviewValueAsDefault=true, inputName=Pre-Skinned Local Position, sortPriority=41}
  n2 = PreSkinnedLocalBounds()
  Divide(A=Subtract(A=@n1, B=@n2.Min), B=Subtract(A=@n2.Max, B=@n2.Max))
  FunctionInput() {bUsePreviewValueAsDefault=true, inputName="Customization Mask (_CM B)", inputType=FunctionInput_Scalar, sortPriority=24}
  n3 = FunctionInput(Preview=MakeMaterialAttributes!(
    BaseColor=Constant() {r=0.18}
    Metallic
    Specular
    Roughness
    Anisotropy
    EmissiveColor
    Opacity
    OpacityMask
    Normal
    Tangent
    WorldPositionOffset
    SubsurfaceColor
    ClearCoat
    ClearCoatRoughness
    AmbientOcclusion
    Refraction
    CustomizedUVs_0
    CustomizedUVs_1
    CustomizedUVs_2
    CustomizedUVs_3
    CustomizedUVs_4
    CustomizedUVs_5
    CustomizedUVs_6
    CustomizedUVs_7
    PixelDepthOffset
    ShadingModel
    Displacement
   )) {inputName=Material, inputType=FunctionInput_MaterialAttributes, sortPriority=-200}
  n4 = FunctionInput(Preview=Constant()) {inputName=isVehicle, inputType=FunctionInput_Scalar}
  n5 = LinearInterpolate(
   A=ScalarParameter() {group=Customization_Drip, parameterName=Drip_Offset_Weapon, sortPriority=-50}
   B=ScalarParameter() {group=Customization_Drip, parameterName=Drip_Offset_Vehicle, sortPriority=-50}
   Alpha=@n4
  )
  n6 = ScalarParameter() {defaultValue=0.2, group=Customization_Drip, parameterName=Drip_Offset_Mech, sortPriority=-50}
  n7 = FunctionInput() {inputName=isMech, inputType=FunctionInput_Scalar}
  n8 = FunctionInput() {bUsePreviewValueAsDefault=true, inputName=Object Bounds Max Adjustment, sortPriority=102}
  n9 = FunctionInput() {bUsePreviewValueAsDefault=true, inputName=Object Bounds Min Adjustment, sortPriority=101}
  n10 = FunctionInput() {bUsePreviewValueAsDefault=true, inputName=Object Bounds Center Adjustment, sortPriority=100}
  n11 = ScalarParameter() {defaultValue=0.5, group=Customization_Drip, parameterName=Drip_Falloff_Weapon, sortPriority=-50}
  n12 = ScalarParameter() {defaultValue=0.5, group=Customization_Drip, parameterName=Drip_Falloff_Vehicle, sortPriority=-50}
  n13 = Divide(
   A=Add(
    A=MaterialFunctionCall!(
     is Vehicle=@n4
     Object Bounds Max Adjustment=@n8
     Object Bounds Min Adjustment=@n9
     Object Bounds Center Adjustment=@n10
     isMech=@n7
     Side Project Vehicle
     Local Position=@n1
    ) {materialFunction=&/MF_WrapBounds}.Result
    B=LinearInterpolate(A=@n5, B=@n6, Alpha=@n7)
   )
   B=LinearInterpolate(
    A=LinearInterpolate(A=@n11, B=@n12, Alpha=@n4)
    B=ScalarParameter() {defaultValue=0.5, group=Customization_Drip, parameterName=Drip_Falloff_Mech, sortPriority=-50}
    Alpha=@n7
   )
  )
  n14 = Clamp(None=@n13)
  n15 = LinearInterpolate(
   A=FunctionInput(Preview=Constant3Vector() {constant=<rgba 0 1 0 0>}) {inputName=ForwardAxis}
   B=VectorParameter() {parameterName=MechForwardAxis, defaultValue=<rgba 0 0 1 1>}.RGB
   Alpha=@n7
  )
  n16 = ScalarParameter() {defaultValue=0.73, group=Customization_Drip, parameterName=DripGradient Power}
  n17 = ScalarParameter() {defaultValue=0.73, group=Customization_Drip, parameterName=DripGradient Power Mech}
  n18 = MaterialFunctionCall(Float3=Power(Base=Multiply(A=@n14, B=@n15), Exp=LinearInterpolate(A=@n16, B=@n17, Alpha=@n7))) {materialFunction=&BreakOutFloat3Components}
  n19 = Saturate(None=Max(A=Max(A=@n18.R, B=@n18.R), B=@n18.B))
  n20 = OneMinus(None=@n19)
  n21 = ScalarParameter() {group=Customization_Drip, parameterName=Dripinvert Gradient}
  n22 = Saturate(None=LinearInterpolate(A=LinearInterpolate(A=@n19, B=@n20, Alpha=@n21), B=@n20, Alpha=@n7))
  n23 = TextureObjectParameter() {
   desc=Roughness in Alpha Channel
   group=Customization_Drip
   parameterName=Wrap_DripDiffuse
   sortPriority=-20
   texture=&T_BlacksRGB
  }
  n24 = TextureObjectParameter() {
   group=Customization_Drip
   parameterName=Wrap_DripNormals
   samplerType=SAMPLERTYPE_Normal
   sortPriority=-19
   texture=&FlatNormalBlue
  }
  n25 = FunctionInput() {bUsePreviewValueAsDefault=true, inputName=Wrap ScaleMultiplier, inputType=FunctionInput_Scalar, previewValue=<1 0 0 1>}
  n26 = Add(
   A=FunctionInput(Preview=Constant3Vector() {constant=<rgba 1 0 0.2 0>}) {bUsePreviewValueAsDefault=true, inputName=PanDirection}
   B=Constant3Vector() {constant=<rgba 0 0 0.2 0>}
  )
  n27 = ScalarParameter() {defaultValue=50, group=Customization_Drip, parameterName=DripSpeed}
  n28 = LinearInterpolate(B=ScalarParameter() {defaultValue=2, group=Customization_Drip, parameterName=VehicleSpeedModulation}, Alpha=@n4) {constA=1}
  n29 = Multiply(
   A=LinearInterpolate(A=@n26, B=VectorParameter() {parameterName=Mech Pan Direction, defaultValue=<rgba 0 0 -1 1>}.RGB, Alpha=@n7)
   B=Multiply(A=Time() {period=1}, B=Multiply(A=@n27, B=@n28))
  )
  n30 = TextureSample() {samplerSource=SSM_Wrap_WorldGroupSettings, texture=&AtmosphericCloudNoise01}
  n31 = LinearInterpolate(
   A=ScalarParameter() {defaultValue=1, group=Customization_Drip, parameterName=DripDistort_Weapon}
   B=ScalarParameter() {defaultValue=10, group=Customization_Drip, parameterName=DripDistort_Vehicle}
   Alpha=@n4
  )
  n32 = PreSkinnedNormal()
  n33 = MaterialFunctionCall!(
   Material Normals
   TextureObject_MASK
   TextureObject_Diffuse=@n23
   TextureObject_Normal=@n24
   TextureScale=Multiply(A=ScalarParameter() {defaultValue=6.5, group=Customization_Drip, parameterName=Wrap_DripTexture Scale, sortPriority=-18}, B=@n25)
   Normal Texture Intensity=ScalarParameter() {defaultValue=1, group=Customization_Drip, parameterName=Wrap_DripNormalIntensity, sortPriority=-17}
   Pre-Skinned Local Position=Add(A=Add(A=@n1, B=@n29), B=Multiply(A=@n30.R, B=@n31))
   Pre-Skinned Local Normal=FunctionInput(Preview=VertexInterpolator(VS=@n32).PS) {bUsePreviewValueAsDefault=true, inputName=Pre-Skinned Local Normal, sortPriority=42}
   Rotation Z Axis
   Rotation X Axis
   Triplanner_Offset
   isMech
   Export Float 4=StaticBool() {value=true}
   ProjectionTransitionContrast=ScalarParameter() {defaultValue=1, group=Customization_Drip, parameterName=Wrap_DripTriplanarSeamHardness, sortPriority=-16}
   IsStatsB
   isCharacter
  ) {materialFunction=/Game/Packages/Fortress_SharedMaterials/Base_Material_Functions/BaseCustomizationFunctions/MF_Base_Customization_Triplanar}
  n34 = StaticSwitchParameter(True=Constant(), False=@n33.Diffuse Texture) {
   desc="puts diffues into emissive channle and fills diffus with black. "
   group=Customization_Drip
   parameterName=Wrap_DripDiffuseColorAsEmissive
  }
  n35 = VectorParameter() {group=Customization_Drip, parameterName=DripGradientColorZLower, sortPriority=31, defaultValue=<rgba 0 0 0 1>}
  n36 = VectorParameter() {group=Customization_Drip, parameterName=DripGradientColorZUpper, sortPriority=31, defaultValue=<rgba 0 0 0 1>}
  n37 = LinearInterpolate(A=@n35.RGB, B=@n36.RGB, Alpha=NamedRerouteUsage())
  n38 = ScalarParameter() {defaultValue=40, group=Customization_Drip, parameterName=DripWeaponDivisor, sortPriority=-40}
  n39 = Multiply(
   A=Divide(
    A=ObjectRadius()
    B=MaterialFunctionCall() {materialFunction=/Engine/Functions/Engine_MaterialFunctions02/WorldPositionOffset/ObjectScale}.Scale X
   )
   B=ScalarParameter() {defaultValue=0.1, group=Customization_Drip, parameterName=DripObjectRadiusScale, sortPriority=-40}
  )
  n40 = ConstantBiasScale(None=Divide(A=@n1, B=LinearInterpolate(A=@n38, B=@n39, Alpha=@n4)))
  n41 = ComponentMask(None=@n40) {b=true}
  n42 = LinearInterpolate(
   A=ScalarParameter() {group=Customization_Drip, parameterName=DripWeaponGradientOffset_Z, sortPriority=30}
   B=ScalarParameter() {group=Customization_Drip, parameterName=DripVehicleGradientOffset_Z, sortPriority=30}
   Alpha=@n4
  )
  n43 = ScalarParameter() {defaultValue=1, group=Customization_Drip, parameterName=DripWeaponGradientPower_Z, sortPriority=30}
  n44 = ScalarParameter() {defaultValue=1, group=Customization_Drip, parameterName=DripVehicleGradientPower_Z, sortPriority=30}
  n45 = Saturate(None=Power(Base=Add(A=@n41, B=@n42), Exp=LinearInterpolate(A=@n43, B=@n44, Alpha=@n4)))
  n46 = StaticSwitch(
   True=Multiply(A=@n33.Diffuse Texture, B=LinearInterpolate(A=@n37, B=@n36.RGB, Alpha=@n45))
   False=@n33.Diffuse Texture
   Value=StaticBoolParameter() {group=Customization_Drip, parameterName=UseGradentAlongZ, sortPriority=29}
  )
  n47 = Constant()
  n48 = VectorParameter() {parameterName=Wrap_DripEmissiveMin, defaultValue=<rgba 1 1 1 1>}
  n49 = VectorParameter() {parameterName=Wrap_DripEmissiveMax, defaultValue=<rgba 1 1 1 1>}
  n50 = DotProduct(A=CameraVectorWS(), B=VertexNormalWS())
  n51 = Saturate(None=LinearInterpolate(
    A=ScalarParameter() {group=Customization_Drip, parameterName=DripEmisViewangleEdge}
    B=ScalarParameter() {defaultValue=1, group=Customization_Drip, parameterName=DripEmisViewangleFace}
    Alpha=Clamp(None=@n50)
   ))
  n52 = ScalarParameter() {group=Customization_Drip, parameterName=Wrap_DripEmissiveClipMin}
  n53 = ScalarParameter() {defaultValue=1, group=Customization_Drip, parameterName=Wrap_DripEmissiveClipMax}
  n54 = Power(Base=@n22, Exp=ScalarParameter() {defaultValue=3, group=Customization_Drip, parameterName=DripEmissiveGradientFalloff})
  n55 = ScalarParameter() {group=Customization_Drip, parameterName=DripEmissiveRampReAdd}
  n56 = Clamp(None=Add(A=LinearInterpolate(A=@n52, B=@n53, Alpha=@n33.Roughness Texture), B=Multiply(A=@n54, B=@n55)))
  n57 = Multiply(A=Multiply(A=LinearInterpolate(A=@n48.RGB, B=@n49.RGB, Alpha=@n33.Roughness Texture), B=Multiply(A=@n51, B=@n56)), B=@n54)
  n58 = GetMaterialAttributes(None=@n3) {attributeGetTypes [#1]}
  n59 = SetMaterialAttributes!(
   MaterialAttributes
   BaseColor=@n34
   Normal=@n33.Normal Texture
   Specular=ScalarParameter() {defaultValue=0.5, group=Customization_Drip, parameterName=Wrap DripSpecular}
   Roughness=LinearInterpolate(
    A=ScalarParameter() {group=Customization_Drip, parameterName=Wrap_DripRoughnessFromDiffuseMin}
    B=ScalarParameter() {defaultValue=1, group=Customization_Drip, parameterName=Wrap_DripRoughnessFromDiffuseMax}
    Alpha=MaterialFunctionCall(Float3=@n33.Diffuse Texture) {materialFunction=&BreakOutFloat3Components}.R
   )
   EmissiveColor=Add(
    A=StaticSwitchParameter(True=@n46, False=@n47) {group=Customization_Drip, parameterName=Wrap_DripDiffuseColorAsEmissive}
    B=Add(A=@n57, B=@n58.EmissiveColor)
   )
   Metallic=ScalarParameter() {group=Customization_Drip, parameterName=Wrap_DripMetallic}
  ) {attributeSetTypes [
    #2
    #3
    #4
    #5
    #1
    #6
   ]}
  n60 = BlendMaterialAttributes(
   A=@n3
   B=@n59
   Alpha=MaterialFunctionCall(
    Transition Phase=@n22
    Height Texture=@n33.Roughness Texture
    Contrast=ScalarParameter() {group=Customization_Drip, parameterName=Wrap_Drip_HeightLerpContrast}
   ) {materialFunction=&HeightLerp}.Alpha
  ) {vertexAttributeBlendType=UseA}
  FunctionOutput(None=StaticSwitchParameter(True=@n60, False=@n3) {group=Customization, parameterName=Drippy Wrap})
  n61 = LinearInterpolate(
   A=ScalarParameter() {defaultValue=-4, group=Customization_Drip, parameterName=DripweaponOffset}
   B=ScalarParameter() {group=Customization_Drip, parameterName=DripVehicleOffset}
   Alpha=@n4
  )
  n62 = ScalarParameter() {group=Customization_Drip, parameterName=DripMechOffset}
  LinearInterpolate(
   A=@n13
   B=Add(A=LinearInterpolate(A=@n61, B=@n62, Alpha=@n7), B=@n40)
   Alpha=ScalarParameter() {group=Customization_Drip, parameterName=offset comparision, sortPriority=-100}
  )
  NamedRerouteDeclaration(None=FunctionInput() {bUsePreviewValueAsDefault=true, inputName=IsCharacter, inputType=FunctionInput_Scalar}) {name=IsCharacter, nodeColor=<rgba 0 0.9253 1 1>}
  n63 = FunctionInput() {bUsePreviewValueAsDefault=true, inputName=IsStatsB, inputType=FunctionInput_Vector2}
  NamedRerouteDeclaration(None=ComponentMask(None=@n63) {r=true}) {name=IsBackpack, nodeColor=<rgba 0 0.9253 1 1>}
  NamedRerouteDeclaration(None=ComponentMask(None=@n63) {g=true}) {name=IsPickaxe, nodeColor=<rgba 0 0.9253 1 1>}
  OneMinus(None=@n26)
  StaticSwitchParameter(True=@n60, False=@n59) {group=Customization, parameterName=Test Wrap}
 ]}
