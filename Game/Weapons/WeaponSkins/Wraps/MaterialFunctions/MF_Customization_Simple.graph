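# Material function for weapon and vehicle wraps that layers a simple secondary material over the incoming attributes within a customization mask.
# Parameters set flat color and surface values, with optional Fresnel-faded emissive and a tiled detail normal, angle-corrected blended at weapon or vehicle tile rates.
# The is_StateB channels select backpack or pickaxe mask scaling, with isCharAccent feeding the backpack term.
# Static switches gate the layer and a following panning emissive accent function.
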MaterialFunction /Game/Weapons/WeaponSkins/Wraps/MaterialFunctions/MF_Customization_Simple
graph {nodes [
  FunctionInput() {bUsePreviewValueAsDefault=true, inputName=--------------------, inputType=FunctionInput_Scalar, sortPriority=40}
  FunctionInput(Preview=Constant()) {bUsePreviewValueAsDefault=true, inputName=------------------, inputType=FunctionInput_Scalar, sortPriority=-19}
  FunctionInput() {
   bUsePreviewValueAsDefault=true
   inputName="Customization Mask UV Islands (_CM B Bottom Right)"
   inputType=FunctionInput_Scalar
   sortPriority=22
  }
  FunctionInput(Preview=Constant()) {bUsePreviewValueAsDefault=true, inputName=-------------------, inputType=FunctionInput_Scalar, sortPriority=23}
  n1 = VectorParameter() {group=Customization_Simple, parameterName=Simple_Color, sortPriority=1, defaultValue=<rgba 0 0 0 1>}
  n2 = ScalarParameter() {defaultValue=20, group=Customization_Simple, parameterName=TileRate_Normal_Weapon, sortPriority=12}
  n3 = ScalarParameter() {defaultValue=40, group=Customization_Simple, parameterName=TileRate_Normal_Vehicle, sortPriority=12}
  n4 = FunctionInput() {inputName=IsVehicle, inputType=FunctionInput_Scalar}
  n5 = TextureSampleParameter2D(UVs=Multiply(A=TextureCoordinate(), B=LinearInterpolate(A=@n2, B=@n3, Alpha=@n4))) {
   group=Customization_Simple
   parameterName=Simple_Normal
   samplerType=SAMPLERTYPE_Normal
   sortPriority=11
   texture=&FlatNormalBlue
  }
  n6 = FunctionInput() {bUsePreviewValueAsDefault=true, inputName=Normal Map, sortPriority=15}
  n7 = VectorParameter() {group=Customization_Simple, parameterName=Simple_Emissive, sortPriority=7, defaultValue=<rgba 0 0 0 1>}
  n8 = Fresnel(ExponentIn=ScalarParameter() {defaultValue=0.5, group=Customization_Simple, parameterName=Simple_Fresnel, sortPriority=8})
  n9 = FunctionInput() {inputName=Material, inputType=FunctionInput_MaterialAttributes, sortPriority=-200}
  n10 = SetMaterialAttributes!(
   MaterialAttributes=@n9
   BaseColor=@n1.RGB
   Metallic=ScalarParameter() {group=Customization_Simple, parameterName=Simple_Metalness, sliderMax=1, sortPriority=4}
   Specular=ScalarParameter() {defaultValue=0.5, group=Customization_Simple, parameterName=Simple_Specular, sliderMax=1, sortPriority=3}
   Roughness=ScalarParameter() {defaultValue=0.5, group=Customization_Simple, parameterName=Simple_Roughness, sliderMax=1, sortPriority=5}
   Opacity=Constant()
   Normal=StaticSwitch(
    True=MaterialFunctionCall(BaseNormal=@n6, AdditionalNormal=@n5.RGB) {materialFunction=&BlendAngleCorrectedNormals}.Result
    False=@n6
    Value=StaticBoolParameter() {group=Customization_Simple, parameterName=UseSimpleNomalMap, sortPriority=10}
   )
   EmissiveColor=StaticSwitch(
    True=LinearInterpolate(A=@n7.RGB, Alpha=@n8) {constB=0}
    False=Constant()
    Value=StaticBoolParameter() {defaultValue=true, group=Customization_Simple, parameterName=UseSimpleEmissive, sortPriority=6}
   )
  ) {attributeSetTypes [
    #1
    #2
    #3
    #4
    #5
    #6
    #7
   ]}
  n11 = FunctionInput() {bUsePreviewValueAsDefault=true, inputName="Customization Mask (_CM B)", inputType=FunctionInput_Scalar, sortPriority=24}
  n12 = ScalarParameter() {defaultValue=1, group=Customization_Simple, parameterName="BackPack Mask Disable 0=active 1=off", sliderMax=1, sortPriority=0}
  n13 = FunctionInput() {bUsePreviewValueAsDefault=true, inputName=isCharAccent, inputType=FunctionInput_Scalar, sortPriority=78}
  n14 = Saturate(None=Add(A=@n12, B=NamedRerouteDeclaration(None=@n13) {name=isCharAccent, nodeColor=<rgba 0 0.9253 1 1>}))
  n15 = MaterialFunctionCall(Float2=FunctionInput() {inputName=is_StateB, inputType=FunctionInput_Vector2}) {materialFunction=&BreakOutFloat2Components}
  n16 = ScalarParameter() {defaultValue=1, group=Customization_Simple, parameterName="Pickaxe Mask Disable 0=active 1=off", sliderMax=1, sortPriority=0}
  n17 = LinearInterpolate(A=LinearInterpolate(A=@n11, B=Multiply(A=@n11, B=@n14), Alpha=@n15.R), B=Multiply(A=@n11, B=@n16), Alpha=@n15.G)
  n18 = StaticSwitchParameter(True=BlendMaterialAttributes(A=@n9, B=@n10, Alpha=@n17), False=@n9) {
   desc=Uses the Top Right corner of the B channel of the _CM Mask to define a simple secondary material
   group=Customization
   parameterName="Allow Secondary Customization Area (Simple)"
   sortPriority=-499
  }
  n19 = FunctionInput() {bUsePreviewValueAsDefault=true, inputName=Pre-Skinned Local Normal, sortPriority=42}
  n20 = FunctionInput() {bUsePreviewValueAsDefault=true, inputName=Pre-Skinned Local Position, sortPriority=41}
  n21 = FunctionInput() {bUsePreviewValueAsDefault=true, inputName="Customization Crevasse (_CM G)", inputType=FunctionInput_Scalar, sortPriority=21}
  FunctionOutput(None=StaticSwitchParameter(
    True=MaterialFunctionCall!(
     Material=@n18
     ------------------
     PreSkinnedLocalNormal=@n19
     PreSkinLocalPosition=@n20
     isVehicle=@n4
     "Customization Crevasse (_CM G)"=@n21
     "Customization Mask (_CM B)"=@n17
    ) {materialFunction=&/MF_Customization_PanningEmissive}.Result
    False=@n18
   ) {group=Customization, parameterName=Panning Emissive Accent})
 ]}
