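# Material function for cosmetic wraps that overlays animated swiping color bands and lightning on an incoming material, behind a static switch.
# Pre-skinned position, normalized by local bounds with vehicle and mech adjustments, feeds rotated, jagged coordinates stepped against looping time to form layered band masks.
# A rotating, world-aligned background texture, projected in local space, fills areas between bands.
# Customization masks limit the BaseColor blend and emissive result, which includes panned lightning textures.
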MaterialFunction /Game/Weapons/WeaponSkins/Wraps/MaterialFunctions/MF_Customization_SwipingPatterns
graph {nodes [
  n1 = FunctionInput(Preview=VertexInterpolator(VS=PreSkinnedLocalBounds().Half Extents).PS) {bUsePreviewValueAsDefault=true, inputName=Local Bounds}
  n2 = MaterialFunctionCall(Float3=@n1) {materialFunction=&BreakOutFloat3Components}
  n3 = MaterialFunctionCall(X=@n2.B, Y=@n2.B, Z=@n2.B) {materialFunction=&MakeFloat3}
  n4 = FunctionInput(Preview=Constant()) {inputName=isVehicle, inputType=FunctionInput_Scalar}
  n5 = FunctionInput() {inputName=isMech, inputType=FunctionInput_Scalar}
  n6 = NamedRerouteUsage()
  n7 = LinearInterpolate(
   A=LinearInterpolate(
    A=LinearInterpolate(A=@n1, B=@n3.Result, Alpha=@n4)
    B=MaterialFunctionCall(X=@n2.R, Y=@n2.R, Z=@n2.R) {materialFunction=&MakeFloat3}.Result
    Alpha=@n5
   )
   B=@n1
   Alpha=@n6
  )
  n8 = FunctionInput(Preview=VertexInterpolator(VS=PreSkinnedPosition()).PS) {bUsePreviewValueAsDefault=true, inputName=Pre-Skinned Local Position, sortPriority=41}
  n9 = MaterialFunctionCall(Float3=@n8) {materialFunction=&BreakOutFloat3Components}
  n10 = MaterialFunctionCall(X=@n9.B, Y=@n9.B, Z=@n9.B) {materialFunction=&MakeFloat3}
  n11 = LinearInterpolate(
   A=LinearInterpolate(
    A=LinearInterpolate(A=@n8, B=@n10.Result, Alpha=@n4)
    B=MaterialFunctionCall(X=@n9.R, Y=@n9.R, Z=@n9.R) {materialFunction=&MakeFloat3}.Result
    Alpha=@n5
   )
   B=@n8
   Alpha=@n6
  )
  n12 = Divide(A=Add(A=@n11, B=@n7), B=Multiply(A=@n7) {constB=5})
  n13 = MaterialFunctionCall(Float3=@n12) {materialFunction=&BreakOutFloat3Components}
  AppendVector(A=@n13.B, B=@n13.B)
  n14 = FunctionInput(Preview=MakeMaterialAttributes!(
    BaseColor
    Metallic
    Specular
    Roughness
    Anisotropy
    EmissiveColor
    Opacity
    OpacityMask
    Normal
    Tangent
    WorldPositionOffset
    SubsurfaceColor
    ClearCoat
    ClearCoatRoughness
    AmbientOcclusion
    Refraction
    CustomizedUVs_0
    CustomizedUVs_1
    CustomizedUVs_2
    CustomizedUVs_3
    CustomizedUVs_4
    CustomizedUVs_5
    CustomizedUVs_6
    CustomizedUVs_7
    PixelDepthOffset
    ShadingModel
    Displacement
   )) {inputName=Material, inputType=FunctionInput_MaterialAttributes, sortPriority=-200}
  n15 = Time()
  n16 = ScalarParameter() {defaultValue=0.4, group=SwipingPatterns, parameterName=SwipingPatterns_Speed, sortPriority=10}
  n17 = Multiply(A=@n15, B=@n16) {constB=0.4}
  MaterialExpressionReroute_2 = Reroute(None=Power(Base=Frac(None=@n17)) {constExponent=0.5})
  n18 = OneMinus(None=@MaterialExpressionReroute_2)
  n19 = Add(
   A=LinearInterpolate(A=ComponentMask(None=@n12) {b=true, g=true}, B=ComponentMask(None=@n12) {b=true, r=true}, Alpha=NamedRerouteUsage())
   B=ScalarParameter() {defaultValue=0.25, group=SwipingPatterns, parameterName=SwipingPatterns_BoundsOffset, sortPriority=11}
  ) {constB=0.25}
  n20 = MaterialFunctionCall(Float2=@n19) {materialFunction=&BreakOutFloat2Components}
  n21 = ComponentMask(None=@n19) {g=true}
  n22 = Frac(None=Multiply(A=@n21) {constB=100})
  n23 = Abs(None=Multiply(A=Subtract(A=@n22) {constB=0.5}) {constB=0.5})
  n24 = Multiply(A=Sine(None=@n15)) {constB=0.1}
  n25 = Add(A=Multiply(A=@n24) {constB=-1}) {constB=0.35}
  n26 = Subtract(A=@n21) {constB=0.025}
  n27 = Subtract(A=@n21) {constB=0.015}
  n28 = Subtract(A=@n21) {constB=-0.045}
  n29 = Frac(None=MaterialFunctionCall(X=Multiply(A=@n26) {constB=10}, Y=Multiply(A=@n27) {constB=5}, Z=Multiply(A=@n28) {constB=5}) {materialFunction=&MakeFloat3}.Result)
  n30 = MaterialFunctionCall(Float3=Round(None=Subtract(A=@n29) {constB=0.45})) {materialFunction=&BreakOutFloat3Components}
  n31 = Multiply(A=Multiply(A=Multiply(A=@n23, B=@n25) {constB=0.5}, B=@n30.G)) {constB=0.55}
  n32 = Constant() {r=1}
  n33 = Rotator(
   Coordinate=MaterialFunctionCall(X=Subtract(A=@n20.R, B=@n31), Y=@n20.G) {materialFunction=&MakeFloat2}.Result
   Time=@n32
  ) {speed=0.075}
  n34 = ComponentMask(None=OneMinus(None=Saturate(None=@n33))) {r=true}
  n35 = Multiply(A=Add(A=@n24) {constB=0.35}, B=@n23) {constB=0.5}
  n36 = Multiply(A=@n35, B=@n30.R)
  n37 = Rotator(
   Coordinate=MaterialFunctionCall(X=Subtract(A=@n20.R, B=@n36), Y=@n20.G) {materialFunction=&MakeFloat2}.Result
   Time=@n32
  ) {speed=0.1}
  n38 = ComponentMask(None=OneMinus(None=Saturate(None=@n37))) {r=true}
  MaterialExpressionAppendVector_4 = AppendVector(A=AppendVector(A=Add(A=@n34) {constB=0.012}, B=Add(A=@n34) {constB=0.28}), B=AppendVector(A=@n38, B=Add(A=@n38) {constB=0.3}))
  n39 = Custom!(a=@MaterialExpressionAppendVector_4, b=@MaterialExpressionReroute_2) {
   code="return step(a,b);"
   description=Step
   outputType=CMOT_Float4
   inputs [
    {
     inputName=a
     input {expression=&self:MaterialExpressionAppendVector_4}
    }
    {
     inputName=b
     input {expression=&self:MaterialExpressionReroute_2}
    }
   ]
  }
  n40 = MaterialFunctionCall(Float4=@n39) {materialFunction=&BreakOutFloat4Components}
  n41 = Multiply(A=Subtract(A=Add(A=@n40.B, B=@n40.B)))
  n42 = Multiply(A=Saturate(None=Multiply(A=@n18, B=@n41))) {constB=-0.25}
  n43 = LinearInterpolate(
   A=ScalarParameter() {defaultValue=5, group=SwipingPatterns, parameterName=Background_TextureScale, sortPriority=18}
   B=ScalarParameter() {defaultValue=7, group=SwipingPatterns, parameterName=Background_TextureScale_Vehicle, sortPriority=18}
   Alpha=@n4
  )
  n44 = FunctionInput(Preview=VertexInterpolator(VS=PreSkinnedNormal()).PS) {bUsePreviewValueAsDefault=true, inputName=Pre-Skinned Local Normal, sortPriority=42}
  n45 = MaterialFunctionCall(Float3=@n44) {materialFunction=&BreakOutFloat3Components}
  n46 = LinearInterpolate(
   A=@n44
   B=MaterialFunctionCall(X=@n45.B, Y=@n45.B, Z=@n45.B) {materialFunction=&MakeFloat3}.Result
   Alpha=@n4
  )
  n47 = MaterialFunctionCall(X=@n45.R, Y=@n45.R, Z=@n45.R) {materialFunction=&MakeFloat3}
  n48 = MaterialFunctionCall!(
   WorldPosition=@n11
   World Space Normal=LinearInterpolate(A=LinearInterpolate(A=@n46, B=@n47.Result, Alpha=@n5), B=@n44, Alpha=@n6)
   UVSize_Front=@n43
   UVSize_Side=@n43
   UVSize_Top=@n43
   ProjectionTransitionContrast=Constant() {r=10}
  ) {materialFunction=/Game/Packages/Fortress_SharedMaterials/Base_Material_Functions/WorldAlignedTexture_UVsMasks}
  n49 = Multiply(A=Time(), B=ScalarParameter() {defaultValue=0.01, group=SwipingPatterns, parameterName=Background_RotationSpeed, sortPriority=25}) {constB=0.01}
  n50 = TextureSampleParameter2D(UVs=Rotator(Coordinate=@n48.UV_Front, Time=@n49) {speed=0.3}) {
   group=SwipingPatterns
   parameterName=Background_Texture
   samplerType=SAMPLERTYPE_Masks
   sortPriority=20
   texture=/Game/Weapons/WeaponSkins/Wraps/Textures/WorldCup2019/WorldCup_Background
  }
  n51 = TextureSampleParameter2D(UVs=Rotator(Coordinate=@n48.UV_Side, Time=@n49) {speed=0.3}) {
   group=SwipingPatterns
   parameterName=Background_Texture
   samplerType=SAMPLERTYPE_Masks
   sortPriority=20
   texture=&WorldCup_Background
  }
  n52 = TextureSampleParameter2D(UVs=Rotator(Coordinate=@n48.UV_Top, Time=@n49) {speed=0.3}) {
   group=SwipingPatterns
   parameterName=Background_Texture
   samplerType=SAMPLERTYPE_Masks
   sortPriority=20
   texture=&WorldCup_Background
  }
  n53 = LinearInterpolate(A=LinearInterpolate(A=@n50.RGBA, B=@n51.RGBA, Alpha=@n48.Mask_Side), B=@n52.RGBA, Alpha=@n48.Mask_Top)
  n54 = ComponentMask(None=@n53) {b=true, g=true, r=true}
  n55 = VectorParameter() {group=SwipingPatterns, parameterName=Background_LineColor, sortPriority=22, defaultValue=<rgba 0.0489 0 0.2344 1>}
  n56 = VectorParameter() {group=SwipingPatterns, parameterName=Background_Color, sortPriority=23, defaultValue=<rgba 0.2086 0 1 1>}
  n57 = Saturate(None=Add(A=Multiply(A=@n54, B=@n55.RGB), B=@n56.RGB))
  n58 = ComponentMask(None=Saturate(None=@n33)) {r=true}
  n59 = Max(A=Multiply(A=Multiply(A=@n35, B=@n30.B)) {constB=1.2}, B=Multiply(A=@n31) {constB=0.4})
  n60 = Rotator(
   Coordinate=MaterialFunctionCall(X=Subtract(A=@n20.R, B=@n59), Y=@n20.G) {materialFunction=&MakeFloat2}.Result
   Time=@n32
  ) {speed=0.11}
  n61 = ComponentMask(None=OneMinus(None=Saturate(None=@n60))) {r=true}
  MaterialExpressionAppendVector_7 = AppendVector(
   A=AppendVector(A=Add(A=@n58) {constB=-0.2}, B=Add(A=@n58) {constB=0.05})
   B=AppendVector(A=Add(A=@n61) {constB=0.1}, B=Add(A=@n61) {constB=0.26})
  )
  n62 = Custom(a=@MaterialExpressionAppendVector_7, b=@MaterialExpressionReroute_2) {
   code="return step(a,b);"
   description=Step
   outputType=CMOT_Float4
   inputs [
    {
     inputName=a
     input {expression=&self:MaterialExpressionAppendVector_7}
    }
    {
     inputName=b
     input {expression=&self:MaterialExpressionReroute_2}
    }
   ]
  }
  n63 = MaterialFunctionCall(Float4=@n62) {materialFunction=&BreakOutFloat4Components}
  n64 = Subtract(A=@n63.B, B=@n63.B)
  n65 = Subtract(A=@n40.R, B=@n40.R)
  n66 = Subtract(A=@n40.B, B=@n40.B)
  n67 = Add(A=@n65, B=@n66)
  n68 = OneMinus(None=Saturate(None=Add(A=@n64, B=@n67)))
  n69 = MaterialFunctionCall(Float3=@n11) {materialFunction=&BreakOutFloat3Components}
  n70 = Multiply(
   A=LinearInterpolate(A=@n69.G, B=Add(A=@n69.G) {constB=80}, Alpha=@n4)
   B=LinearInterpolate(A=Constant() {r=0.05}, B=Constant() {r=0.015}, Alpha=@n4)
  ) {constB=0.05}
  n71 = ComponentMask(None=@n53) {a=true}
  n72 = Power(Base=Saturate(None=Add(A=@n70, B=@n71))) {constExponent=4}
  n73 = FunctionInput() {bUsePreviewValueAsDefault=true, inputName="Customization Mask 1 (_CM B Top Left)", inputType=FunctionInput_Scalar, sortPriority=25}
  n74 = FunctionInput() {bUsePreviewValueAsDefault=true, inputName="Customization Mask 2 (_CM B Top Right)", inputType=FunctionInput_Scalar, sortPriority=26}
  n75 = Add(A=@n73, B=@n74)
  n76 = Saturate(None=@n75)
  n77 = BlendMaterialAttributes(
   A=@n14
   B=SetMaterialAttributes!(MaterialAttributes=@n14, BaseColor=Add(A=@n42, B=@n57)) {attributeSetTypes [#1]}
   Alpha=Multiply(A=Multiply(A=@n68, B=@n72), B=@n76)
  ) {vertexAttributeBlendType=UseA}
  n78 = MaterialFunctionCall(Float2=@n19) {materialFunction=&BreakOutFloat2Components}
  n79 = Multiply(A=Subtract(A=Time()) {constB=-0.7}, B=@n16) {constB=0.4}
  n80 = Frac(None=@n79)
  n81 = Frac(None=Panner(Coordinate=@n78.R, Time=Power(Base=@n80)) {speedX=1})
  n82 = Subtract(A=Multiply(A=@n78.G) {constB=20}, B=LinearInterpolate(A=Constant() {r=12}, B=Constant() {r=10.5}, Alpha=@n4)) {constB=12}
  n83 = Multiply(A=@n78.R) {constB=16}
  n84 = Multiply(A=@n78.R) {constB=4}
  n85 = Multiply(A=@n78.R) {constB=2.5}
  n86 = Subtract(A=Frac(None=MaterialFunctionCall(X=@n83, Y=@n84, Z=@n85) {materialFunction=&MakeFloat3}.Result)) {constB=0.5}
  n87 = MaterialFunctionCall(Float3=Abs(None=Multiply(A=@n86) {constB=2})) {materialFunction=&BreakOutFloat3Components}
  n88 = Add(A=Multiply(A=@n87.R), B=Add(A=Multiply(A=@n87.G), B=@n87.B))
  n89 = TextureSampleParameter2D(UVs=MaterialFunctionCall(X=Multiply(A=@n81) {constB=10}, Y=Add(A=@n82, B=@n88)) {materialFunction=&MakeFloat2}.Result) {
   group=SwipingPatterns
   parameterName=Lightning_Texture
   samplerSource=SSM_Clamp_WorldGroupSettings
   samplerType=SAMPLERTYPE_Masks
   sortPriority=50
   texture=/Game/Weapons/WeaponSkins/Wraps/Textures/WorldCup2019/WorldCup_Lightning
  }
  n90 = MaterialFunctionCall(X=Multiply(A=@n78.R) {constB=6}, Y=Add(A=Multiply(A=@n88) {constB=0.25}, B=Multiply(A=@n78.G) {constB=6})) {materialFunction=&MakeFloat2}
  n91 = Multiply(A=@n79) {constB=4, desc=2}
  n92 = TextureSampleParameter2D(UVs=Panner(Coordinate=Multiply(A=@n90.Result) {constB=2}, Time=Round(None=@n91)) {speedX=-0.1, speedY=0.1}) {
   group=SwipingPatterns
   parameterName=Lightning_SecondaryTexture
   samplerType=SAMPLERTYPE_Masks
   sortPriority=60
   texture=/Game/Weapons/WeaponSkins/Wraps/Textures/WorldCup2019/WorldCup_LightningSecondary
  }
  n93 = Multiply(A=Multiply(A=@n89.A) {constB=5}, B=@n92.R) {constB=10}
  n94 = Max(A=@n89.RGB, B=Round(None=Saturate(None=@n93)))
  n95 = Saturate(None=Add(A=@n75, B=@n4))
  n96 = Saturate(None=@n64)
  n97 = Multiply(A=@n96, B=VectorParameter() {group=SwipingPatterns, parameterName=Layer2_Color, sortPriority=2, defaultValue=<rgba 1 0 0.15 1>}.RGB)
  n98 = Subtract(A=@n63.R, B=@n63.R)
  n99 = OneMinus(None=@n96)
  n100 = Multiply(A=Multiply(A=Saturate(None=@n98), B=@n65), B=@n99)
  n101 = VectorParameter() {group=SwipingPatterns, parameterName=Layer3_Crossover_Color, sortPriority=3, defaultValue=<rgba 0.0648 0 0.3005 1>}
  n102 = Saturate(None=Subtract(A=Subtract(A=@n65, B=@n64), B=@n98))
  n103 = VectorParameter() {group=SwipingPatterns, parameterName=Layer1_Color, sortPriority=1, defaultValue=<rgba 0.0273 1 0.2582 1>}
  n104 = Saturate(None=Subtract(A=@n66, B=@n65))
  n105 = Add(
   A=Add(A=Multiply(A=@n100, B=@n101.RGB), B=Multiply(A=@n102, B=@n103.RGB))
   B=Multiply(A=Multiply(A=@n99, B=@n104), B=VectorParameter() {group=SwipingPatterns, parameterName=Layer0_Color, defaultValue=<rgba 0.7758 1 0 1>}.RGB)
  )
  n106 = Frac(None=@n20.G)
  n107 = Multiply(A=Multiply(A=Subtract(A=@n106) {constB=0.5}) {constB=2}, B=@n76)
  n108 = Multiply(A=Add(A=Add(A=@n97, B=@n105), B=Multiply(A=@n107) {constB=0.15}), B=@n72)
  n109 = VectorParameter() {group=SwipingPatterns, parameterName=Background_EmissiveLineColor, sortPriority=24, defaultValue=<rgba 0.0609 0 0.2917 1>}
  n110 = Multiply(A=@n68, B=Multiply(A=Multiply(A=@n109.RGB, B=@n54), B=Saturate(None=@n70)))
  n111 = Add(A=Multiply(A=Multiply(A=@n94) {constB=2}, B=@n95), B=Add(A=Multiply(A=@n108, B=@n95), B=Multiply(A=@n110, B=@n76)))
  FunctionOutput(None=StaticSwitchParameter(
    True=SetMaterialAttributes!(MaterialAttributes=@n77, EmissiveColor=@n111) {attributeSetTypes [#2]}
    False=GetMaterialAttributes(None=@n14).MaterialAttributes
   ) {group=Customization_Wrap, parameterName=SwipingPatterns Wrap})
  MaterialFunctionCall() {materialFunction=&LocalPosition}
  Multiply(A=@n76)
  NamedRerouteDeclaration(None=FunctionInput() {bUsePreviewValueAsDefault=true, inputName=IsCharacter, inputType=FunctionInput_Scalar}) {name=IsCharacter, nodeColor=<rgba 0 0.9253 1 1>}
  n112 = FunctionInput() {bUsePreviewValueAsDefault=true, inputName=IsStatsB, inputType=FunctionInput_Vector2}
  NamedRerouteDeclaration(None=ComponentMask(None=@n112) {r=true}) {name=IsBackpack, nodeColor=<rgba 0 0.9253 1 1>}
  NamedRerouteDeclaration(None=ComponentMask(None=@n112) {g=true}) {name=IsPickaxe, nodeColor=<rgba 0 0.9253 1 1>}
  Transform(None=VertexNormalWS()) {transformSourceType=TRANSFORMSOURCE_World, transformType=TRANSFORM_Local}
 ]}
