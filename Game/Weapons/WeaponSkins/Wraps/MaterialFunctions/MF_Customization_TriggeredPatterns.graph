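# Wrap material function that adds Blueprint-triggered bursts of spinning, trailing stars to emissive, plus star backgrounds to masked base color, behind a static switch.
# Pre-skinned position, normalized by bounds and adjusted per item type, supplies projection UVs and a side mask.
# Three timed, hash-seeded bursts get random LUT colors.
# A Custom node injects shared HLSL functions by closing its own body early; loop count drops on cheaper quality and shading paths.
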MaterialFunction /Game/Weapons/WeaponSkins/Wraps/MaterialFunctions/MF_Customization_TriggeredPatterns
graph {nodes [
  n1 = GetMaterialAttributes(None=FunctionInput(Preview=MakeMaterialAttributes!(
     BaseColor
     Metallic
     Specular
     Roughness
     Anisotropy
     EmissiveColor
     Opacity
     OpacityMask
     Normal
     Tangent
     WorldPositionOffset
     SubsurfaceColor
     ClearCoat
     ClearCoatRoughness
     AmbientOcclusion
     Refraction
     CustomizedUVs_0
     CustomizedUVs_1
     CustomizedUVs_2
     CustomizedUVs_3
     CustomizedUVs_4
     CustomizedUVs_5
     CustomizedUVs_6
     CustomizedUVs_7
     PixelDepthOffset
     ShadingModel
     Displacement
    )) {inputName=Material, inputType=FunctionInput_MaterialAttributes, sortPriority=-200}) {attributeGetTypes [#1, #2]}
  MaterialExpressionScalarParameter_0 = ScalarParameter() {defaultValue=1, group=TriggeredPatterns, parameterName=TriggeredPatterns_Speed, sortPriority=10}
  n2 = Divide(B=@MaterialExpressionScalarParameter_0)
  n3 = Time()
  n4 = ScalarParameter() {defaultValue=999999, group=TriggeredPatterns, parameterName=GameTimeFromBP_Iteration1}
  MaterialExpressionSubtract_1 = Subtract(A=@n3, B=@n4)
  n5 = ScalarParameter() {defaultValue=1, group=TriggeredPatterns, parameterName=CharacterScale}
  n6 = ScalarParameter() {defaultValue=1, group=TriggeredPatterns, parameterName=BackpackScale}
  n7 = NamedRerouteUsage()
  n8 = FunctionInput(Preview=VertexInterpolator(VS=PreSkinnedPosition()).PS) {bUsePreviewValueAsDefault=true, inputName=Pre-Skinned Local Position, sortPriority=41}
  n9 = MaterialFunctionCall(Float3=@n8) {materialFunction=&BreakOutFloat3Components}
  n10 = FunctionInput(Preview=Constant()) {inputName=isVehicle, inputType=FunctionInput_Scalar}
  n11 = LinearInterpolate(
   A=@n8
   B=MaterialFunctionCall(X=@n9.B, Y=@n9.B, Z=@n9.B) {materialFunction=&MakeFloat3}.Result
   Alpha=@n10
  )
  n12 = MaterialFunctionCall(X=@n9.R, Y=@n9.R, Z=@n9.R) {materialFunction=&MakeFloat3}
  n13 = FunctionInput(Preview=Constant()) {bUsePreviewValueAsDefault=true, inputName=isMech, inputType=FunctionInput_Scalar}
  n14 = NamedRerouteUsage()
  n15 = Multiply(
   A=LinearInterpolate(A=LinearInterpolate(A=@n11, B=@n12.Result, Alpha=@n13), B=@n8, Alpha=@n14)
   B=LinearInterpolate(B=LinearInterpolate(A=@n5, B=@n6, Alpha=@n7), Alpha=NamedRerouteUsage()) {constA=1}
  ) {constB=0.2}
  n16 = PreSkinnedLocalBounds()
  n17 = MaterialFunctionCall(Float3=@n16.Min) {materialFunction=&BreakOutFloat3Components}
  n18 = LinearInterpolate(
   A=LinearInterpolate(
    A=@n16.Min
    B=MaterialFunctionCall(X=@n17.B, Y=@n17.B, Z=@n17.B) {materialFunction=&MakeFloat3}.Result
    Alpha=@n10
   )
   B=MaterialFunctionCall(X=@n17.R, Y=@n17.R, Z=@n17.R) {materialFunction=&MakeFloat3}.Result
   Alpha=@n13
  )
  n19 = FunctionInput(Preview=VertexInterpolator(VS=@n16.Half Extents).PS) {bUsePreviewValueAsDefault=true, inputName=Local Bounds}
  n20 = MaterialFunctionCall(Float3=@n19) {materialFunction=&BreakOutFloat3Components}
  n21 = LinearInterpolate(
   A=@n19
   B=MaterialFunctionCall(X=@n20.B, Y=@n20.B, Z=@n20.B) {materialFunction=&MakeFloat3}.Result
   Alpha=@n10
  )
  n22 = MaterialFunctionCall(X=@n20.R, Y=@n20.R, Z=@n20.R) {materialFunction=&MakeFloat3}
  n23 = Multiply(A=LinearInterpolate(A=LinearInterpolate(A=@n21, B=@n22.Result, Alpha=@n13), B=@n19, Alpha=@n14)) {constB=2}
  n24 = ComponentMask(None=@n23) {r=true}
  n25 = Max(A=ComponentMask(None=@n23) {g=true}, B=ComponentMask(None=@n23) {b=true})
  n26 = Divide(A=Subtract(A=@n15, B=LinearInterpolate(A=@n18, B=@n16.Min, Alpha=@n14)), B=AppendVector(A=AppendVector(A=@n24, B=@n25), B=@n25))
  n27 = ComponentMask(None=@n26) {b=true, g=true}
  n28 = LinearInterpolate(
   A=ComponentMask(None=@n26) {b=true, desc=Character, r=true}
   B=ComponentMask(None=@n26) {b=true, desc=Backpack, r=true}
   Alpha=NamedRerouteUsage()
  )
  n29 = NamedRerouteUsage()
  n30 = VectorParameter() {group=TriggeredPatterns, parameterName=TriggeredPatterns_BoundsOffset_XY, defaultValue=<rgba 0 0.275 0 1>}
  MaterialExpressionReroute_1 = Reroute(None=Add(A=LinearInterpolate(A=@n27, B=@n28, Alpha=@n29), B=ComponentMask(None=@n30.RGB) {g=true, r=true}) {constB=0.25})
  n31 = Custom!(In=@MaterialExpressionReroute_1) {
   code="return In;\r\n}\r\n\r\n// Triangles\r\nfloat TriggeredPatternsTriangleFunction(float2 pTriangle)\r\n{\r\n    pTriangle.y -= 0.4f;\r\n\tfloat x = 1.414213562373095f/2.0f * (pTriangle.x - pTriangle.y);\r\n\tfloat y = 1.414213562373095f/2.0f * (pTriangle.x + pTriangle.y);\r\n\tfloat r1 = max(abs(x), abs(y)) - 2.0f/(2.0f*1.414213562373095f);\r\n\tfloat r2 = pTriangle.y;\r\n\tfloat c = max(r1,r2);\r\n\treturn (c < 0.01) ? 1 : 0;\r\n}\r\n\r\n\r\n// Stars\r\nfloat TriggeredPatternsStarFunction(float2 P, float2 RotationCosSin)\r\n{\r\n\tfloat Star = 0.0f;\r\n\t\r\n\tfloat C = RotationCosSin.x;\r\n\tfloat S = RotationCosSin.y;\r\n\r\n\tfloat2x2 Mat0 = float2x2( C, S, -S, C);\r\n\tfloat2x2 Mat1 = float2x2( -S, C, -C, -S);\r\n\tfloat2x2 Mat2 = float2x2( -C, -S, S, -C);\r\n\tfloat2x2 Mat3 = float2x2( S, -C, C, S);\r\n\t\r\n\tStar += TriggeredPatternsTriangleFunction(mul(P, Mat0));\r\n\tStar += TriggeredPatternsTriangleFunction(mul(P, Mat1));\r\n\tStar += TriggeredPatternsTriangleFunction(mul(P, Mat2));\r\n\tStar += TriggeredPatternsTriangleFunction(mul(P, Mat3));\r\n\r\n\tStar = saturate(Star);\r\n\r\n\treturn Star;\r\n}\r\n\r\n// Stars and Trails\r\nfloat TriggeredPatternsGetResult(float InTime, float2 InUV, float InScaleStart, float InScaleEnd, float InSpeed, float3 InHash, float LoopIterations)\r\n{\r\n    float Shape = 0.0f;\r\n\t\r\n\tfloat2 Rotations[4];\r\n\tRotations[0] = float2(cos(InTime*4.651), sin(InTime*4.651));\r\n\tRotations[1] = float2(cos(InTime*-4.115), sin(InTime*-4.115));\r\n\tRotations[2] = float2(cos(InTime*2.712), sin(InTime*2.712));\r\n\tRotations[3] = float2(cos(InTime*-3.446), sin(InTime*-3.446));\r\n\t\r\n\tfor (int s = 0; s < LoopIterations; s++)\t\r\n\t{\r\n\t\tfloat Rand = frac(((InHash.x+2.3871f)*(s + 5.314f)) * (5.452313f + InHash.y + InHash.z)) * 2.0f - 1.0f;\r\n\t\tfloat Rand2 = frac(((InHash.y+5.3871f)*(s + 2.314f)) * (7.152454f + InHash.x + InHash.z)) * 2.0f - 1.0f;\r\n\t\t\r\n\t\tfloat TimeScaled = InTime * InSpeed;\r\n\t\tfloat TimeScaledPow = TimeScaled * TimeScaled;\r\n\t\t\r\n\t\tfloat ParticleVMax = frac((s + 5.813f + Rand) * 3.645f) * 2.0f - 1.0f;\r\n\t\tfloat ParticleV = frac(lerp(0.5f + (Rand * 0.25f), 0.5f + 0.5f * ParticleVMax, TimeScaledPow));\r\n\t\tfloat ParticleU = 0.8f - TimeScaled + (Rand2 * 0.33f);\r\n\t\t\r\n\t\tfloat2 DistTri = float2(ParticleU, ParticleV) - float2(InUV.x, InUV.y);\r\n\t\tfloat2 DistTrail = DistTri;\r\n\t\tDistTri *= lerp((InScaleStart + Rand*10.0f), (InScaleEnd + Rand*2.5f), TimeScaled);\r\n\t\t\t\t\r\n\t\tfloat SignMultiplier = s && 4 ? 1 : -1;\r\n\t\tuint RandomRotationIndex = (uint)floor(abs(Rand * 4));\r\n\t\tfloat2 RandomRotation = Rotations[RandomRotationIndex && 3] * SignMultiplier;\r\n\r\n\t\t//Stars\r\n\t\tShape += TriggeredPatternsStarFunction(DistTri, RandomRotation) * (1.0f - Shape);\r\n\t\t\r\n\t\t//Trails\r\n\t\tfloat TrailsMask = 1.0f - ceil(saturate(DistTrail.x));\r\n\t\tfloat TrailsTail = saturate(DistTrail.x + (0.25f + (TimeScaledPow * 0.25))) ;\r\n\t\tShape = max(Shape, (1.0f - saturate(abs(lerp(0.5f + (Rand * 0.25f), 0.5f + 0.5f * ParticleVMax, (0.8f - InUV.x + (Rand2 * 0.33f)) * TimeScaled) - InUV.y) * 80.0f)) * TrailsMask * TrailsTail * 1.0f);\r\n\r\n\t\t// Fade out opacity\r\n\t\tShape *= 1.0f - TimeScaledPow;\r\n\t\t\r\n\t\tShape = saturate(Shape);\r\n\t}\r\n\r\n    return saturate(Shape);\r\n}\r\n\r\nvoid TriggeredPatternsDummy()\r\n{\r\n\r\n\r\n/*\r\n\r\n//To use in a following custom code node to use the declared functions above. Workaround for ES3.1\r\n\r\nreturn TriggeredPatternsGetResult(Time, UV, ScaleStart, ScaleEnd, Speed, Hash, LoopIterations);\r\n\r\n*/"
   description=DeclareTriggeredPatternsFunctions
   outputType=CMOT_Float2
   inputs [{
     inputName=In
     input {expression=&self:MaterialExpressionReroute_1}
    }]
  }
  n32 = VectorParameter() {group=TriggeredPatterns, parameterName=Pliant_Star_UV_Adjust_Pickaxe_Male}
  n33 = VectorParameter() {group=TriggeredPatterns, parameterName=Pliant_Star_UV_Adjust_Pickaxe_Female}
  n34 = ScalarParameter() {bUseCustomPrimitiveData=true, parameterName=Pliant Pickaxe Type}
  n35 = Add(
   A=LinearInterpolate(A=ComponentMask(None=@n32.RGB) {g=true, r=true}, B=ComponentMask(None=@n33.RGB) {g=true, r=true}, Alpha=Subtract(A=@n34) {constB=3})
   B=@n31
  )
  n36 = NamedRerouteUsage()
  n37 = ComponentMask(None=VectorParameter() {group=TriggeredPatterns, parameterName=Pliant_Star_UV_Adjust_Character}.RGB) {g=true, r=true}
  MaterialExpressionReroute_14 = Reroute(None=LinearInterpolate(A=LinearInterpolate(A=@n31, B=@n35, Alpha=@n36), B=Add(A=@n31, B=@n37), Alpha=NamedRerouteUsage()))
  MaterialExpressionScalarParameter_3 = ScalarParameter() {defaultValue=50, group=TriggeredPatterns, parameterName=TriggeredPatterns_Scale_Start}
  MaterialExpressionScalarParameter_4 = ScalarParameter() {defaultValue=5, group=TriggeredPatterns, parameterName=TriggeredPatterns_Scale_End}
  n38 = Frac(None=@n4)
  n39 = Frac(None=Multiply(A=VectorParameter() {parameterName=TriggeredPatterns_RandomHashFromBP_Iteration1}.RGB) {constB=0.0741})
  MaterialExpressionAdd_8 = Add(A=Multiply(A=@n38) {constB=121.2413}, B=Multiply(A=@n39) {constB=121.6541})
  n40 = Ceil(None=ScalarParameter() {defaultValue=10, group=TriggeredPatterns, parameterName=TriggeredPatterns_LoopIterations})
  n41 = Max(A=@n40)
  n42 = Max(A=Divide(A=@n40))
  MaterialExpressionQualitySwitch_0 = QualitySwitch(Default=ShadingPathSwitch(Default=@n41, Deferred=@n41, Forward=@n42, Mobile=@n42), Low=@n42)
  n43 = Custom!(
   Time=@MaterialExpressionSubtract_1
   Speed=@MaterialExpressionScalarParameter_0
   UV=@MaterialExpressionReroute_14
   ScaleStart=@MaterialExpressionScalarParameter_3
   ScaleEnd=@MaterialExpressionScalarParameter_4
   Hash=@MaterialExpressionAdd_8
   LoopIterations=@MaterialExpressionQualitySwitch_0
  ) {
   code="return TriggeredPatternsGetResult(Time, UV, ScaleStart, ScaleEnd, Speed, Hash, LoopIterations);"
   description=Loop_StarsPatterns
   outputType=CMOT_Float1
   inputs [
    {
     inputName=Time
     input {expression=&self:MaterialExpressionSubtract_1}
    }
    {
     inputName=Speed
     input {expression=&self:MaterialExpressionScalarParameter_0}
    }
    {
     inputName=UV
     input {expression=&self:MaterialExpressionReroute_14}
    }
    {
     inputName=ScaleStart
     input {expression=&self:MaterialExpressionScalarParameter_3}
    }
    {
     inputName=ScaleEnd
     input {expression=&self:MaterialExpressionScalarParameter_4}
    }
    {
     inputName=Hash
     input {expression=&self:MaterialExpressionAdd_8}
    }
    {
     inputName=LoopIterations
     input {expression=&self:MaterialExpressionQualitySwitch_0}
    }
   ]
  }
  n44 = Constant()
  n45 = If(A=@n2, B=@MaterialExpressionSubtract_1, "A > B"=@n43, "A == B"=@n44, "A < B"=@n44)
  n46 = MaterialFunctionCall(Float3=@n39) {materialFunction=&BreakOutFloat3Components}
  n47 = Add(A=@n46.R, B=@n46.R)
  n48 = Frac(None=Add(A=@n38, B=Add(A=@n47, B=@n46.B)))
  n49 = TextureObjectParameter() {group=TriggeredPatterns, parameterName=TriggeredPatterns_ColorLUT, texture=&T_WhitesRGB}
  n50 = Multiply(A=@n45, B=TextureSample(UVs=AppendVector(A=@n38, B=@n48), Tex=@n49).RGB)
  n51 = ScalarParameter() {defaultValue=999999, group=TriggeredPatterns, parameterName=GameTimeFromBP_Iteration2}
  MaterialExpressionSubtract_2 = Subtract(A=@n3, B=@n51)
  n52 = Frac(None=@n51)
  n53 = Frac(None=Multiply(A=VectorParameter() {parameterName=TriggeredPatterns_RandomHashFromBP_Iteration2}.RGB) {constB=0.1239})
  MaterialExpressionAdd_15 = Add(A=Multiply(A=@n52) {constB=164.8123}, B=Multiply(A=@n53) {constB=110.3215})
  n54 = Custom(
   Time=@MaterialExpressionSubtract_2
   Speed=@MaterialExpressionScalarParameter_0
   UV=@MaterialExpressionReroute_14
   ScaleStart=@MaterialExpressionScalarParameter_3
   ScaleEnd=@MaterialExpressionScalarParameter_4
   Hash=@MaterialExpressionAdd_15
   LoopIterations=@MaterialExpressionQualitySwitch_0
  ) {
   code="return TriggeredPatternsGetResult(Time, UV, ScaleStart, ScaleEnd, Speed, Hash, LoopIterations);"
   description=Loop_StarsPatterns
   outputType=CMOT_Float1
   inputs [
    {
     inputName=Time
     input {expression=&self:MaterialExpressionSubtract_2}
    }
    {
     inputName=Speed
     input {expression=&self:MaterialExpressionScalarParameter_0}
    }
    {
     inputName=UV
     input {expression=&self:MaterialExpressionReroute_14}
    }
    {
     inputName=ScaleStart
     input {expression=&self:MaterialExpressionScalarParameter_3}
    }
    {
     inputName=ScaleEnd
     input {expression=&self:MaterialExpressionScalarParameter_4}
    }
    {
     inputName=Hash
     input {expression=&self:MaterialExpressionAdd_15}
    }
    {
     inputName=LoopIterations
     input {expression=&self:MaterialExpressionQualitySwitch_0}
    }
   ]
  }
  n55 = Constant()
  n56 = If(A=@n2, B=@MaterialExpressionSubtract_2, "A > B"=@n54, "A == B"=@n55, "A < B"=@n55)
  n57 = Subtract(B=@n56)
  n58 = MaterialFunctionCall(Float3=@n53) {materialFunction=&BreakOutFloat3Components}
  n59 = Add(A=@n52, B=Add(A=Add(A=@n58.R, B=@n58.R), B=@n58.B))
  n60 = TextureSample(UVs=AppendVector(A=@n52, B=Frac(None=@n59)), Tex=@n49)
  n61 = ScalarParameter() {defaultValue=999999, group=TriggeredPatterns, parameterName=GameTimeFromBP_Iteration3}
  MaterialExpressionSubtract_3 = Subtract(A=@n3, B=@n61)
  n62 = Frac(None=@n61)
  n63 = Frac(None=Multiply(A=VectorParameter() {parameterName=TriggeredPatterns_RandomHashFromBP_Iteration3}.RGB) {constB=0.1515})
  MaterialExpressionAdd_16 = Add(A=Multiply(A=@n62) {constB=90.5461}, B=Multiply(A=@n63) {constB=130.871})
  n64 = Custom(
   Time=@MaterialExpressionSubtract_3
   Speed=@MaterialExpressionScalarParameter_0
   UV=@MaterialExpressionReroute_14
   ScaleStart=@MaterialExpressionScalarParameter_3
   ScaleEnd=@MaterialExpressionScalarParameter_4
   Hash=@MaterialExpressionAdd_16
   LoopIterations=@MaterialExpressionQualitySwitch_0
  ) {
   code="return TriggeredPatternsGetResult(Time, UV, ScaleStart, ScaleEnd, Speed, Hash, LoopIterations);"
   description=Loop_StarsPatterns
   outputType=CMOT_Float1
   inputs [
    {
     inputName=Time
     input {expression=&self:MaterialExpressionSubtract_3}
    }
    {
     inputName=Speed
     input {expression=&self:MaterialExpressionScalarParameter_0}
    }
    {
     inputName=UV
     input {expression=&self:MaterialExpressionReroute_14}
    }
    {
     inputName=ScaleStart
     input {expression=&self:MaterialExpressionScalarParameter_3}
    }
    {
     inputName=ScaleEnd
     input {expression=&self:MaterialExpressionScalarParameter_4}
    }
    {
     inputName=Hash
     input {expression=&self:MaterialExpressionAdd_16}
    }
    {
     inputName=LoopIterations
     input {expression=&self:MaterialExpressionQualitySwitch_0}
    }
   ]
  }
  n65 = Constant()
  n66 = If(A=@n2, B=@MaterialExpressionSubtract_3, "A > B"=@n64, "A == B"=@n65, "A < B"=@n65)
  n67 = Multiply(A=Add(A=Multiply(A=@n50, B=@n57), B=Multiply(A=@n56, B=@n60.RGB)), B=Subtract(B=@n66))
  n68 = MaterialFunctionCall(Float3=@n63) {materialFunction=&BreakOutFloat3Components}
  n69 = Add(A=Add(A=@n68.R, B=@n68.R), B=@n68.B)
  n70 = AppendVector(A=@n62, B=Frac(None=Add(A=@n62, B=@n69)))
  n71 = Multiply(A=Multiply(A=@n66, B=TextureSample(UVs=@n70, Tex=@n49).RGB), B=Subtract(B=@n45))
  n72 = ScalarParameter() {defaultValue=2, group=TriggeredPatterns, parameterName=TriggeredPatterns_EmissiveScale}
  n73 = Constant() {r=1}
  n74 = MaterialFunctionCall(Value_Morning=@n73, Value_Day=Constant() {r=1.5}, Value_Evening=@n73, Value_Night=Constant() {r=0.5}) {materialFunction=&Scalar_Time_of_Day}
  n75 = FunctionInput(Preview=VertexInterpolator(VS=PreSkinnedNormal()).PS) {bUsePreviewValueAsDefault=true, inputName=Pre-Skinned Local Normal, sortPriority=42}
  n76 = MaterialFunctionCall(Float3=@n75) {materialFunction=&BreakOutFloat3Components}
  n77 = MaterialFunctionCall(X=@n76.B, Y=@n76.B, Z=@n76.B) {materialFunction=&MakeFloat3}
  n78 = LinearInterpolate(
   A=LinearInterpolate(
    A=LinearInterpolate(A=@n75, B=@n77.Result, Alpha=@n10)
    B=MaterialFunctionCall(X=@n76.R, Y=@n76.R, Z=@n76.R) {materialFunction=&MakeFloat3}.Result
    Alpha=@n13
   )
   B=@n75
   Alpha=@n14
  )
  n79 = MaterialFunctionCall(Float3=@n78) {materialFunction=&BreakOutFloat3Components}
  n80 = MaterialFunctionCall(X=@n79.G, Y=@n79.G, Z=@n79.G) {materialFunction=&MakeFloat3}
  n81 = NamedRerouteUsage()
  n82 = MaterialFunctionCall!(
   WorldPosition=@n15
   World Space Normal=LinearInterpolate(A=@n78, B=LinearInterpolate(A=@n80.Result, B=@n80.Result, Alpha=@n81), Alpha=NamedRerouteUsage())
   UVSize_Front
   UVSize_Side
   UVSize_Top
   ProjectionTransitionContrast=Constant() {r=0.5}
  ) {materialFunction=/Game/Packages/Fortress_SharedMaterials/Base_Material_Functions/WorldAlignedTexture_UVsMasks}
  n83 = LinearInterpolate(A=@n1.EmissiveColor, B=Multiply(A=Add(A=@n67, B=@n71), B=Multiply(A=@n72, B=@n74.Result)) {constB=2}, Alpha=@n82.Mask_Side)
  n84 = TextureSample(
   UVs=Rotator(Coordinate=@MaterialExpressionReroute_1, Time=Constant() {r=-1})
   Tex=TextureObjectParameter() {group=TriggeredPatterns, parameterName=TriggeredPatterns_Background, texture=&T_BlacksRGB}
  )
  n85 = Time()
  n86 = Rotator(
   Coordinate=Panner(Coordinate=Multiply(A=@MaterialExpressionReroute_1) {constB=10}, Time=@n85) {speedX=0.015, speedY=0.0075}
   Time=Constant() {r=-1.5}
  )
  n87 = TextureObjectParameter() {group=TriggeredPatterns, parameterName=TriggeredPatterns_BackgroundStars, texture=&T_BlacksRGB}
  n88 = Panner(Coordinate=Multiply(A=@MaterialExpressionReroute_1) {constB=4}, Time=@n85) {speedX=1.8, speedY=0.0075}
  n89 = TextureSample(UVs=Multiply(A=@n88, B=Constant2Vector() {g=-1, r=1}), Tex=@n87)
  n90 = MaterialFunctionCall!(
   UVs_Layer1=@n88
   UVs_Layer2
   Seed_Layer1=Constant() {r=3.6514}
   Seed_Layer2
   Threshold_Layer1=Constant() {r=0.3325}
   Threshold_Layer2
   EdgeMaskWidth_Layer1
   EdgeMaskWidth_Layer2
  ) {materialFunction=/Game/Weapons/WeaponSkins/MaterialFunction/MF_RandomTilePatternFromUVs}
  n91 = Multiply(A=@n90.Layer1_Mask, B=@n90.Layer1_Mask)
  n92 = Multiply(A=Add(A=TextureSample(UVs=@n86, Tex=@n87).RGB, B=Multiply(A=@n89.A, B=@n91)), B=@n82.Mask_Side)
  n93 = FunctionInput() {bUsePreviewValueAsDefault=true, inputName="Customization Mask 1 (_CM B Top Left)", inputType=FunctionInput_Scalar, sortPriority=25}
  n94 = FunctionInput() {bUsePreviewValueAsDefault=true, inputName="Customization Mask 2 (_CM B Top Right)", inputType=FunctionInput_Scalar, sortPriority=26}
  n95 = LinearInterpolate(A=@n1.BaseColor, B=Add(A=@n84.RGB, B=Multiply(A=@n92, B=@n84.A)), Alpha=Saturate(None=Add(A=@n93, B=@n94)))
  FunctionOutput(None=StaticSwitchParameter(
    True=SetMaterialAttributes!(MaterialAttributes=@n1.MaterialAttributes, EmissiveColor=@n83, BaseColor=@n95) {attributeSetTypes [#1, #2]}
    False=@n1.MaterialAttributes
   ) {group=Customization_Wrap, parameterName=TriggeredPatterns Wrap})
  NamedRerouteDeclaration(None=FunctionInput() {bUsePreviewValueAsDefault=true, inputName=IsCharacter, inputType=FunctionInput_Scalar}) {name=IsCharacter, nodeColor=<rgba 0 0.9253 1 1>}
  n96 = FunctionInput() {bUsePreviewValueAsDefault=true, inputName=IsStatsB, inputType=FunctionInput_Vector2}
  NamedRerouteDeclaration(None=ComponentMask(None=@n96) {r=true}) {name=IsBackpack, nodeColor=<rgba 0 0.9253 1 1>}
  NamedRerouteDeclaration(None=ComponentMask(None=@n96) {g=true}) {name=IsPickaxe, nodeColor=<rgba 0 0.9253 1 1>}
  ObjectBounds()
  Reroute(None=AppendVector(A=@MaterialExpressionReroute_1, B=Constant()))
 ]}
