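# Weapon wrap material function that adds animated digital-rain glyphs to incoming emissive, with optional overrides for base color, metallic, specular and roughness.
# Triplanar or mesh UVs drive randomized flipbook glyphs under stepped, panning falling gradients, with a second layer on High/Epic quality.
# Glyph color tints the rain plus optional glow, mask and fresnel, then fake sun/moon specular is added.
# Static switches gate the effect, customization-mask blending and a Winterfest animation.
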MaterialFunction /Game/Weapons/WeaponSkins/Wraps/MaterialFunctions/MF_DigitalRain
graph {nodes [
  n1 = SetMaterialAttributes!(MaterialAttributes)
  n2 = GetMaterialAttributes(None=NamedRerouteDeclaration(None=FunctionInput(Preview=@n1) {bUsePreviewValueAsDefault=true, inputName=MA_In, inputType=FunctionInput_MaterialAttributes}) {name=MA_In, nodeColor=<rgba 1 0.7734 0 1>}) {attributeGetTypes [
    #1
    #2
    #3
    #4
    #5
    #6
   ]}
  n3 = SetMaterialAttributes!(
   MaterialAttributes=@n2.MaterialAttributes
   EmissiveColor=Add(A=@n2.EmissiveColor, B=NamedRerouteUsage())
   BaseColor=NamedRerouteUsage()
   Metallic=NamedRerouteUsage()
   Specular=NamedRerouteUsage()
   Roughness=NamedRerouteUsage()
  ) {attributeSetTypes [
    #1
    #2
    #3
    #4
    #5
   ]}
  n4 = StaticSwitchParameter(
   True=MaterialFunctionCall!(In=@n3) {desc=Used for winterfest, materialFunction=/Game/Environments/FrontEnd/MaterialFunctions/MF_CozyLodge_Anim}.Result
   False=@n3
  ) {group=DigitalRain, parameterName=UseCozyLodgeAnim}
  n5 = FunctionInput() {bUsePreviewValueAsDefault=true, inputName=CustomizationMask, inputType=FunctionInput_Vector4, sortPriority=1, previewValue=<1 1 1 1>}
  n6 = BlendMaterialAttributes(A=NamedRerouteUsage(), B=@n4, Alpha=ChannelMaskParameter(None=@n5) {group=DigitalRain, parameterName=DigitalRainCustomizationMaskChannel}.R)
  FunctionOutput(None=StaticSwitchParameter(
    True=StaticSwitchParameter(True=@n6, False=@n4) {group=DigitalRain, parameterName=DigitalRainUseCustomizationMask}
    False=NamedRerouteUsage()
   ) {group=Customization_Wrap, parameterName=UseDigitalRain})
  n7 = ScalarParameter() {defaultValue=6, group=DigitalRain, parameterName=MainUVScale}
  NamedRerouteDeclaration(None=@n7) {name=MainUVScale, nodeColor=<rgba 1 0 0.4453 1>}
  n8 = NamedRerouteUsage()
  n9 = NamedRerouteUsage()
  n10 = Time()
  n11 = Multiply(A=NamedRerouteUsage(), B=Multiply(A=NamedRerouteUsage(), B=Constant() {r=8}))
  n12 = Multiply(A=Time(), B=ScalarParameter() {defaultValue=0.1647, group=DigitalRain, parameterName=RainCharacterGlyphRowSpeed})
  n13 = Constant() {r=8}
  n14 = Divide(A=Ceil(None=Multiply(A=@n12, B=@n13)), B=@n13)
  n15 = Multiply(
   A=ScalarParameter() {defaultValue=0.365, group=DigitalRain, parameterName=RainGlyphSwitchSpeed}
   B=MaterialFunctionCall!(
    PrimaryTextureCoordinates=Panner(Coordinate=@n11, Time=@n14)
    2ndOrderTextureCoordinates
    RandomSeed=ScalarParameter() {defaultValue=4, group=DigitalRain, parameterName=RainGlyphShiftRandomSeed}
    NoiseDensityBias
    Bypass Noise Influence Crosstalk
   ) {materialFunction=/Game/Effects/Fort_Effects/Materials/_Functions/MF_RandomUVSample}.NoiseWithRange
  )
  n16 = Constant() {r=8}
  n17 = Ceil(None=Multiply(A=Multiply(A=@n10, B=@n15), B=@n16))
  n18 = NamedRerouteUsage()
  n19 = TextureSample!(
   UVs=Panner(Coordinate=Multiply(A=@n8, B=@n9), Time=Divide(A=@n17, B=@n16)) {bFractionalPart=true, speedX=1}
   Tex
   "DDX(UVs)"=DDX(Value=@n18)
   "DDY(UVs)"=DDY(Value=@n18)
   Apply View MipBias
  ) {mipValueMode=TMVM_Derivative, texture=/Game/Weapons/WeaponSkins/Wraps/Textures/DigitalRain/T_DigitalRain_Flipbook_01_SubPixelEdge}
  n20 = Multiply(A=ScalarParameter() {defaultValue=5, group=DigitalRain, parameterName=RainGlyphFresnelEdgeScale}, B=@n19.RGB)
  n21 = MaterialFunctionCall(Exponent=ScalarParameter() {defaultValue=4, group=DigitalRain, parameterName=RainGlyphFresnelExponent}) {materialFunction=&MF_Fresnel_FX}
  n22 = Multiply(A=AppendVector(A=NamedRerouteUsage(), B=Constant() {r=1}), B=NamedRerouteUsage())
  n23 = MaterialFunctionCall(PrimaryTextureCoordinates=@n22, RandomSeed=ScalarParameter() {defaultValue=8, group=DigitalRain, parameterName=RainRandomSeed}) {materialFunction=&MF_RandomUVSample}
  n24 = Multiply(A=Constant() {r=8}, B=NamedRerouteUsage())
  n25 = Floor(None=Multiply(A=NamedRerouteDeclaration(None=@n23.NoiseWithRange) {name=RainGradients, nodeColor=<rgba 0.4766 0 1 1>}, B=@n24) {constB=8})
  n26 = Multiply(A=Constant() {r=8}, B=NamedRerouteUsage())
  n27 = Multiply(A=Multiply(A=Time(), B=ScalarParameter() {defaultValue=-0.4, group=DigitalRain, parameterName=RainPanSpeed}), B=@n26)
  n28 = Panner(Coordinate=Add(A=@n22, B=Divide(A=@n25, B=@n24) {constB=8}), Time=Divide(A=Ceil(None=@n27), B=@n26)) {speedY=1}
  n29 = SphereMask(
   A=Frac(None=MaterialFunctionCall(Float2=@n28) {materialFunction=&BreakOutFloat2Components}.G)
   B=ScalarParameter() {defaultValue=1, group=DigitalRain, parameterName=RainGradientCenterPoint}
   Radius=ScalarParameter() {defaultValue=0.4, group=DigitalRain, parameterName=RainGradientRadius}
   Hardness=Constant()
  ) {desc=Final rain gradient}
  n30 = ScalarParameter() {defaultValue=6, group=DigitalRain, parameterName=RainGradientExponent}
  n31 = Multiply(
   A=Multiply(A=LinearInterpolate(A=@n20, B=@n19.RGB, Alpha=@n21.Fresnel), B=Power(Base=@n29, Exp=@n30))
   B=Power(Base=NamedRerouteUsage(), Exp=ScalarParameter() {defaultValue=2, group=DigitalRain, parameterName=RainGradientExponents})
  )
  n32 = Add(A=@n31, B=NamedRerouteUsage())
  n33 = Multiply(A=QualitySwitch(Default=@n31, High=@n32, Epic=@n32), B=NamedRerouteUsage())
  n34 = StaticSwitchParameter(True=Add(A=NamedRerouteUsage(), B=@n33), False=@n33) {group=DigitalRain, parameterName=UseGlowSphericalGradient}
  n35 = TextureSampleParameter2D() {group=DigitalRain, parameterName=RainMaskTexture, samplerType=SAMPLERTYPE_Masks, texture=&T_WhiteMask}
  n36 = StaticSwitchParameter(True=Multiply(A=@n34, B=@n35.RGB), False=@n34) {group=DigitalRain, parameterName=UseMaskAgainstRain}
  n37 = MaterialFunctionCall!(
   Thickness=ScalarParameter() {defaultValue=0.1, group=DigitalRain, parameterName=FresnelThickness}
   NormalWS=Transform(None=NamedRerouteUsage())
   PowerExponent=ScalarParameter() {defaultValue=1.5, group=DigitalRain, parameterName=FresnelExponent}
  ) {materialFunction=/Game/Weapons/Prototype/ObjectMover/Materials/MF_NormalizedFresnel}
  n38 = ScalarParameter() {defaultValue=0.1, group=DigitalRain, parameterName=FresnelIntensity}
  n39 = StaticSwitchParameter(True=Add(A=@n36, B=Multiply(A=@n37.PowerInverted, B=@n38)), False=@n36) {defaultValue=true, group=DigitalRain, parameterName=UseNormalizedFresnel}
  n40 = VectorParameter() {group=DigitalRain, parameterName=GlyphColor, defaultValue=<rgba 1.2396 6 2.4297 0>}
  n41 = SphereMask(
   A=NamedRerouteUsage()
   B=ScalarParameter() {defaultValue=0.85, group=DigitalRain, parameterName=GradientCenterPoint}
   Radius=ScalarParameter() {defaultValue=1, group=DigitalRain, parameterName=GradientRadius}
   Hardness=ScalarParameter() {defaultValue=0.2, group=DigitalRain, parameterName=GradientHardness}
  )
  n42 = OneMinus(None=@n41)
  n43 = StaticSwitchParameter(
   True=LinearInterpolate(
    A=VectorParameter() {group=DigitalRain, parameterName=GlyphColorAlt, defaultValue=<rgba 1.3715 2 0.1146 0>}.RGB
    B=@n40.RGB
    Alpha=StaticSwitchParameter(True=@n42, False=@n41) {group=DigitalRain, parameterName=InvertGradientForGlyphColor}
   )
   False=@n40.RGB
  ) {group=DigitalRain, parameterName=UseGradientForGlyphColor}
  NamedRerouteDeclaration(None=Add(A=Multiply(A=@n39, B=@n43), B=NamedRerouteUsage())) {name=DigitalRain, nodeColor=<rgba 0.8738 1 0.0406 1>}
  NamedRerouteDeclaration(None=@n2.Metallic) {name=MetallicIn, nodeColor=<rgba 1 0 0.0937 1>}
  NamedRerouteDeclaration(None=@n2.Specular) {name=SpecularIn, nodeColor=<rgba 1 0.7969 0 1>}
  NamedRerouteDeclaration(None=@n2.Roughness) {name=RoughnessIn, nodeColor=<rgba 0.5938 0 1 1>}
  NamedRerouteDeclaration(None=StaticSwitchParameter(True=ScalarParameter() {defaultValue=1, group=DigitalRain, parameterName=DigitalRainMetallic}, False=NamedRerouteUsage()) {group=DigitalRain, parameterName=OverrideMetallic}) {name=MetallicOut, nodeColor=<rgba 1 0.2344 0 1>}
  NamedRerouteDeclaration(None=StaticSwitchParameter(True=ScalarParameter() {defaultValue=0.05, group=DigitalRain, parameterName=DigitalRainSpecular}, False=NamedRerouteUsage()) {group=DigitalRain, parameterName=OverrideSpecular}) {name=SpecularOut, nodeColor=<rgba 0.9688 0 1 1>}
  NamedRerouteDeclaration(None=StaticSwitchParameter(True=ScalarParameter() {group=DigitalRain, parameterName=DigitalRainRoughness}, False=NamedRerouteUsage()) {group=DigitalRain, parameterName=OverrideRoughness}) {name=RoughnessOut, nodeColor=<rgba 0.9922 1 0 1>}
  NamedRerouteDeclaration(None=@n2.Normal) {name=NormalIn, nodeColor=<rgba 0 1 0.5547 1>}
  n44 = Transform(None=NamedRerouteUsage())
  n45 = CollectionParameter() {
   collection=&FortniteMaterialParameters
   parameterName=SunAndMoonModelDirectionalVector
  }
  n46 = Saturate(None=DotProduct(A=ReflectionVectorWS(CustomWorldNormal=@n44) {bNormalizeCustomWorldNormal=true}, B=ComponentMask(None=@n45) {b=true, g=true, r=true}))
  n47 = ScalarParameter() {defaultValue=10, group=DigitalRain, parameterName=FakeSpecularExponent}
  NamedRerouteDeclaration(None=Multiply(
    A=Power(Base=@n46, Exp=@n47)
    B=VectorParameter() {group=DigitalRain, parameterName=FakeSpecularColor, defaultValue=<rgba 0.0175 0.03 0.0024 0>}.RGB
   )) {name=FakeSpecular, nodeColor=<rgba 0 1 0.2734 1>}
  n48 = PreSkinnedNormal() {desc=Switched to Preskinned Vertex Normals to prevent rotation problems on pliant character}
  n49 = MaterialFunctionCall(Float3=Abs(None=VertexInterpolator(VS=@n48).PS)) {materialFunction=&BreakOutFloat3Components}
  NamedRerouteDeclaration(None=If(A=@n49.B, B=Max(A=@n49.R, B=@n49.R), "A > B"=Constant(), "A < B"=Constant() {r=1})) {name=LocalSpaceVectorMask, nodeColor=<rgba 0.3828 1 0 1>}
  n50 = NamedRerouteUsage()
  n51 = NamedRerouteUsage()
  n52 = Time()
  n53 = Multiply(A=NamedRerouteUsage(), B=Multiply(A=NamedRerouteUsage(), B=Constant() {r=8}))
  n54 = Multiply(A=Time(), B=ScalarParameter() {defaultValue=0.4381, group=DigitalRain, parameterName=RainCharacterGlyphRowSpeedEPIC})
  n55 = Constant() {r=8}
  n56 = Divide(A=Ceil(None=Multiply(A=@n54, B=@n55)), B=@n55)
  n57 = Multiply(
   A=ScalarParameter() {defaultValue=0.88, group=DigitalRain, parameterName=RainGlyphSwitchSpeedEPIC}
   B=MaterialFunctionCall(
    PrimaryTextureCoordinates=Panner(Coordinate=@n53, Time=@n56)
    RandomSeed=ScalarParameter() {defaultValue=6, group=DigitalRain, parameterName=RainGlyphShiftRandomSeedEPIC}
   ) {materialFunction=&MF_RandomUVSample}.NoiseWithRange
  )
  n58 = Constant() {r=8}
  n59 = Ceil(None=Multiply(A=Multiply(A=@n52, B=@n57), B=@n58))
  n60 = TextureSample(
   UVs=Panner(Coordinate=Multiply(A=@n50, B=@n51), Time=Divide(A=@n59, B=@n58)) {bFractionalPart=true, speedX=1}
   "DDX(UVs)"=DDX(Value=@n51)
   "DDY(UVs)"=DDY(Value=@n51)
  ) {mipValueMode=TMVM_Derivative, texture=&T_DigitalRain_Flipbook_01_SubPixelEdge}
  n61 = LinearInterpolate(
   A=Multiply(A=ScalarParameter() {defaultValue=2, group=DigitalRain, parameterName=RainGlyphFresnelEdgeScaleEPIC}, B=@n60.RGB)
   B=@n60.RGB
   Alpha=MaterialFunctionCall(Exponent=ScalarParameter() {defaultValue=4, group=DigitalRain, parameterName=RainGlyphFresnelExponentEPIC}) {materialFunction=&MF_Fresnel_FX}.Fresnel
  )
  n62 = Multiply(A=AppendVector(A=NamedRerouteUsage(), B=Constant() {r=1}), B=NamedRerouteUsage())
  n63 = MaterialFunctionCall(PrimaryTextureCoordinates=@n62, RandomSeed=ScalarParameter() {defaultValue=12, group=DigitalRain, parameterName=RainRandomSeedEPIC}) {materialFunction=&MF_RandomUVSample}
  n64 = Multiply(A=Constant() {r=8}, B=NamedRerouteUsage())
  n65 = Divide(A=Floor(None=Multiply(A=@n63.NoiseWithRange, B=@n64) {constB=8}), B=@n64) {constB=8}
  n66 = Time()
  n67 = ScalarParameter() {defaultValue=-0.35, group=DigitalRain, parameterName=RainPanSpeedEPIC}
  n68 = Multiply(A=Constant() {r=8}, B=NamedRerouteUsage())
  n69 = Ceil(None=Multiply(A=Multiply(A=@n66, B=@n67), B=@n68))
  n70 = MaterialFunctionCall(Float2=Panner(Coordinate=Add(A=@n62, B=@n65), Time=Divide(A=@n69, B=@n68)) {bFractionalPart=true, speedY=1}) {materialFunction=&BreakOutFloat2Components}
  n71 = Power(
   Base=SphereMask(
    A=Frac(None=@n70.G)
    B=ScalarParameter() {defaultValue=1, group=DigitalRain, parameterName=RainGradientCenterPointEPIC}
    Radius=ScalarParameter() {defaultValue=0.3, group=DigitalRain, parameterName=RainGradientRadiusEPIC}
    Hardness=Constant()
   ) {desc=Final rain gradient}
   Exp=ScalarParameter() {defaultValue=12, group=DigitalRain, parameterName=RainGradientExponentEPIC}
  )
  NamedRerouteDeclaration(None=Multiply(A=Multiply(A=@n61, B=@n71), B=ScalarParameter() {defaultValue=0.35, group=DigitalRain, parameterName=EpicScale})) {name=EpicRain, nodeColor=<rgba 1 0 0.6094 1>}
  n72 = AppendVector(A=Constant() {r=8}, B=Constant() {r=1})
  n73 = MaterialFunctionCall!(
   UseLocalSpace=StaticBool() {value=true}
   WorldPosition=WorldPosition() {worldPositionShaderOffset=WPT_ExcludeAllShaderOffsets}.XYZ
   NormalOverride
  ) {materialFunction=/Game/Environments/Apollo/Sets/Skyfire/Materials/MaterialFunctions/MF_WorldAlignedUVs_HardEdge}
  n74 = Min(
   A=InverseLinearInterpolate(
    A=ScalarParameter() {defaultValue=1.3, group=DigitalRain, parameterName=ObjectRadiusScaleRemapMin}
    B=ScalarParameter() {defaultValue=2.5, group=DigitalRain, parameterName=ObjectRadiusScaleRemapMax}
    Value=ObjectRadius()
   )
   B=ScalarParameter() {defaultValue=400, group=DigitalRain, parameterName=ObjectRadiusDivisorMax}
  )
  n75 = ScalarParameter() {defaultValue=400, group=DigitalRain, parameterName=ObjectRadiusDivisor}
  n76 = StaticSwitchParameter(
   True=Divide(
    A=NamedRerouteDeclaration(None=@n73.Result) {name=TriPlanar.XYZ, nodeColor=<rgba 1 0 0.2578 1>}
    B=StaticSwitchParameter(True=@n74, False=@n75) {defaultValue=true, group=DigitalRain, parameterName=ObjectRadiusScalesDivisor}
   )
   False=TextureCoordinate()
  ) {defaultValue=true, group=DigitalRain, parameterName=UseCheapTriplanarMapping}
  n77 = StaticSwitchParameter(True=MaterialFunctionCall(XY=@n76) {materialFunction=&Swizzle}.YX, False=@n76) {group=DigitalRain, parameterName=SwizzleUVs}
  NamedRerouteDeclaration(None=Multiply(A=@n72, B=NamedRerouteDeclaration(None=@n77) {name=UV.Epic, nodeColor=<rgba 0.5469 0 1 1>})) {name=UV.8x1.EPIC, nodeColor=<rgba 1 0.2344 0 1>}
  NamedRerouteDeclaration(None=@n7) {name=MainUVScaleEPIC, nodeColor=<rgba 1 0.5859 0 1>}
  NamedRerouteDeclaration(None=Multiply(A=NamedRerouteDeclaration(None=@n77) {name=UV, nodeColor=<rgba 0.0312 1 0 1>}, B=@n72)) {name=UV.8x1, nodeColor=<rgba 1 0.0234 0 1>}
  n78 = ChannelMaskParameter(None=Divide(
    A=PreSkinnedPosition()
    B=MaterialFunctionCall() {materialFunction=&ObjectLocalBounds}.Local Bounds Size
   )) {group=DigitalRain, maskChannel=Blue, parameterName=0-1 Gradient Channel, defaultValue=<rgba 0 0 1 0>}
  NamedRerouteDeclaration(None=VertexInterpolator(VS=@n78.R).PS) {name=Bounds.0-1.Z, nodeColor=<rgba 0 0 1 1>}
  NamedRerouteDeclaration(None=@n2.BaseColor) {name=BaseColorIn, nodeColor=<rgba 0.8984 0 1 1>}
  n79 = WorldPosition()
  n80 = VectorParameter() {group=DigitalRain, parameterName=GlowGradientCenter}
  n81 = SphereMask(
   A=StaticSwitchParameter(True=TransformPosition(None=@n79.XYZ) {transformSourceType=TRANSFORMPOSSOURCE_World}, False=NamedRerouteUsage()) {group=DigitalRain, parameterName=UseCheapTriplanarMapping}
   B=StaticSwitchParameter(True=@n80.RGB, False=AppendVector(A=@n80.R, B=@n80.R)) {group=DigitalRain, parameterName=UseCheapTriplanarMapping}
   Radius=ScalarParameter() {defaultValue=1, group=DigitalRain, parameterName=GlowGradientRadius}
   Hardness=ScalarParameter() {group=DigitalRain, parameterName=GlowGradientHardness}
  )
  n82 = VectorParameter() {group=DigitalRain, parameterName=GlowGradientColor, defaultValue=<rgba 0.0948 0.3595 0.5052 0>}
  NamedRerouteDeclaration(None=Power(Base=Multiply(A=@n81, B=@n82.RGB), Exp=ScalarParameter() {defaultValue=2.5, group=DigitalRain, parameterName=GlowGradientExponent})) {name=SphericalGradient, nodeColor=<rgba 1 0 0.8906 1>}
  NamedRerouteDeclaration(None=StaticSwitchParameter(
    True=VectorParameter() {group=DigitalRain, parameterName=DigitalRainBaseColor, defaultValue=<rgba 0.01 0.01 0.01 0>}.RGB
    False=NamedRerouteUsage()
   ) {group=DigitalRain, parameterName=OverrideBaseColor}) {name=BaseColorOut, nodeColor=<rgba 0.6875 1 0 1>}
  VertexNormalWS()
 ]}
