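# Material function for weapon wraps that layers optional edge wear and dirt over input material attributes, with separate tuning for weapons, vehicles and mechs.
# Crevasse input drives edge wear and GMap drives dirt, both broken up by world-aligned noise projected on pre-skinned position and normal.
# Dirt is suppressed where edge wear shows.
# Static switches choose flat or textured layers, shared mask texture and which customization-mask regions receive the effect.
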MaterialFunction /Game/Weapons/WeaponSkins/Wraps/MaterialFunctions/MF_DirtAndEdgeWearMasked
graph {nodes [
  n1 = TextureSample() {samplerType=SAMPLERTYPE_Masks, texture=/Game/Weapons/FORT_Rifles/Textures/T_Scar_CM}
  n2 = FunctionInput(Preview=Constant() {r=1}) {bUsePreviewValueAsDefault=true, inputName=isVehicle, inputType=FunctionInput_Scalar}
  n3 = LinearInterpolate(
   A=ScalarParameter() {defaultValue=0.95, group=DirtAndEdgeWear, parameterName=DE_WeaponCurvatureSpread}
   B=ScalarParameter() {defaultValue=0.95, group=DirtAndEdgeWear, parameterName=DE_VehicleCurvatureSpread}
   Alpha=@n2
  )
  n4 = ScalarParameter() {defaultValue=0.75, group=DirtAndEdgeWear, parameterName=DE_MechCurvatureSpread}
  n5 = FunctionInput(Preview=Constant()) {bUsePreviewValueAsDefault=true, inputName=isMech, inputType=FunctionInput_Scalar}
  n6 = ScalarParameter() {defaultValue=2, group=DirtAndEdgeWear, parameterName=DE_WeaponCurvature_Strength}
  n7 = ScalarParameter() {defaultValue=2, group=DirtAndEdgeWear, parameterName=DE_VehicleCurvature_Strength}
  n8 = Multiply(
   A=Subtract(
    A=FunctionInput(Preview=@n1.G) {bUsePreviewValueAsDefault=true, inputName=Crevasse, inputType=FunctionInput_Scalar}
    B=LinearInterpolate(A=@n3, B=@n4, Alpha=@n5)
   )
   B=LinearInterpolate(
    A=LinearInterpolate(A=@n6, B=@n7, Alpha=@n2)
    B=ScalarParameter() {defaultValue=2, group=DirtAndEdgeWear, parameterName=DE_MechCurvature_Strength}
    Alpha=@n5
   )
  )
  n9 = Clamp(None=@n8)
  n10 = VectorParameter() {group=DirtAndEdgeWear, parameterName=DE_WeaponEdgeWearGrungeTexture_Tiling, defaultValue=<rgba 12 12 12 1>}
  n11 = VectorParameter() {group=DirtAndEdgeWear, parameterName=DE_VehicleEdgeWearGrungeTexture_Tiling, defaultValue=<rgba 12 12 12 1>}
  n12 = VectorParameter() {group=DirtAndEdgeWear, parameterName=DE_MechEdgeWearGrungeTexture_Tiling, defaultValue=<rgba 12 12 12 1>}
  n13 = FunctionInput(Preview=VertexInterpolator(VS=PreSkinnedPosition()).PS) {bUsePreviewValueAsDefault=true, inputName=PreskinnedPosition}
  n14 = FunctionInput(Preview=VertexInterpolator(VS=PreSkinnedNormal()).PS) {bUsePreviewValueAsDefault=true, inputName=PreskinnedNormal}
  n15 = MaterialFunctionCall(
   TextureObject=NamedRerouteUsage()
   TextureSize=LinearInterpolate(A=LinearInterpolate(A=@n10.RGB, B=@n11.RGB, Alpha=@n2), B=@n12.RGB, Alpha=@n5)
   WorldPosition=@n13
   Export Float 4=StaticBool() {value=true}
   World Space Normal=@n14
  ) {materialFunction=&WorldAlignedTexture}
  n16 = LinearInterpolate(
   A=ScalarParameter() {defaultValue=1, group=DirtAndEdgeWear, parameterName=DE_EdgeWearGrungeNoiseStrength}
   B=ScalarParameter() {defaultValue=1, group=DirtAndEdgeWear, parameterName=DE_VehicleEdgeWearGrungeNoiseStrength}
   Alpha=@n2
  )
  n17 = ScalarParameter() {defaultValue=1, group=DirtAndEdgeWear, parameterName=DE_MechEdgeWearGrungeNoiseStrength}
  n18 = Subtract(
   A=@n9
   B=Multiply(
    A=ChannelMaskParameter(None=@n15.XYZ Texture) {group=DirtAndEdgeWear, parameterName=DE_EdgeWearGrungeTextureChannel}.R
    B=LinearInterpolate(A=@n16, B=@n17, Alpha=@n5)
   )
  )
  Ceil(None=@n18)
  n19 = LinearInterpolate(
   A=LinearInterpolate(
    A=VectorParameter() {group=DirtAndEdgeWear, parameterName=DE_WeaponBottomLayer_Tiling, defaultValue=<rgba 512 512 512 1>}.RGB
    B=VectorParameter() {group=DirtAndEdgeWear, parameterName=DE_VehicleBottomLayer_Tiling, defaultValue=<rgba 512 512 512 1>}.RGB
    Alpha=@n2
   )
   B=VectorParameter() {group=DirtAndEdgeWear, parameterName=DE_MechBottomLayer_Tiling, defaultValue=<rgba 512 512 512 1>}.RGB
   Alpha=@n5
  )
  n20 = Add(A=@n13, B=VectorParameter() {group=DirtAndEdgeWear, parameterName=DE_BottomLayer_Offset, defaultValue=<rgba 0 0 0 1>}.RGB)
  n21 = StaticBool() {value=true}
  ComponentMask(None=MaterialFunctionCall(TextureObject=NamedRerouteUsage(), TextureSize=@n19, WorldPosition=@n20, Export Float 4=@n21, World Space Normal=@n14) {materialFunction=&WorldAlignedTexture}.XYZ Texture) {b=true, g=true, r=true}
  Constant()
  FunctionInput() {bUsePreviewValueAsDefault=true, inputName=TopLayerMaterial_Custom, inputType=FunctionInput_MaterialAttributes}
  FunctionInput() {bUsePreviewValueAsDefault=true, inputName=BottomLayerMaterial_Custom, inputType=FunctionInput_MaterialAttributes}
  n22 = VectorParameter() {group=DirtAndEdgeWear, parameterName=DE_BottomLayerSimpleBaseColor, defaultValue=<rgba 0.1172 0.1778 0.25 1>}
  n23 = LinearInterpolate(
   A=NamedRerouteUsage()
   B=NamedRerouteUsage()
   Alpha=ScalarParameter() {group=DirtAndEdgeWear, parameterName="DE_BottomLayer Normals 0=on 1=Removed"}
  )
  n24 = GetMaterialAttributes(None=FunctionInput() {bUsePreviewValueAsDefault=true, inputName=Material, inputType=FunctionInput_MaterialAttributes}) {attributeGetTypes [#1]}
  n25 = LinearInterpolate(
   A=VectorParameter() {group=DirtAndEdgeWear, parameterName=DE_WeaponBottomLayer_Rotation, defaultValue=<rgba 0 0 0 1>}.RGB
   B=VectorParameter() {group=DirtAndEdgeWear, parameterName=DE_VehicleBottomLayer_Rotation, defaultValue=<rgba 0 0 0 1>}.RGB
   Alpha=@n2
  )
  n26 = VectorParameter() {group=DirtAndEdgeWear, parameterName=DE_MechBottomLayer_Rotation, defaultValue=<rgba 0 0 0 1>}
  n27 = MaterialFunctionCall(Float3=Divide(A=LinearInterpolate(A=@n25, B=@n26.RGB, Alpha=@n5))) {materialFunction=&BreakOutFloat3Components}
  n28 = MaterialFunctionCall(
   Export Float 4=@n21
   TextureObject=NamedRerouteUsage()
   TextureSize=@n19
   WorldPosition=@n20
   World Space Normal=@n14
   RotZAxis=@n27.B
   RotXAxis=@n27.B
   RotYAxis=@n27.B
  ) {materialFunction=&/WorldAlignedTexture}
  n29 = StaticSwitchParameter(
   True=SetMaterialAttributes!(
    MaterialAttributes=@n24.MaterialAttributes
    BaseColor=@n22.RGB
    Roughness=ScalarParameter() {defaultValue=0.15, group=DirtAndEdgeWear, parameterName=DE_BottomLayerSimpleRoughness}
    Normal=@n23
   ) {attributeSetTypes [#2, #3, #1]}
   False=SetMaterialAttributes(
    MaterialAttributes=@n24.MaterialAttributes
    BaseColor=ComponentMask(None=@n28.XYZ Texture) {b=true, g=true, r=true}
    Roughness=ComponentMask(None=@n28.XYZ Texture) {a=true}
    Normal=@n23
   ) {attributeSetTypes [#2, #3, #1]}
  ) {defaultValue=true, group=DirtAndEdgeWear, parameterName=UseSimpleMaterialBottomLayer}
  n30 = ScalarParameter() {defaultValue=1, group=DirtAndEdgeWear, parameterName=DE_BottomLayer_Metalness}
  n31 = ScalarParameter() {defaultValue=1, group=DirtAndEdgeWear, parameterName=DE_BottomLayer_Specular}
  n32 = ScalarParameter() {defaultValue=0.01, group=DirtAndEdgeWear, parameterName="DE_EdgeWear A  0=hard----->1=Soft", sortPriority=0}
  n33 = StaticSwitchParameter(True=@n9, False=@n18) {group=DirtAndEdgeWear, parameterName="DE_EdgeWearMask---->Just Crevase", sortPriority=10}
  n34 = Constant() {r=1}
  n35 = VectorParameter() {group=DirtAndEdgeWear, parameterName=DE_WeaponEdgeWearMaskTexture_Tiling, defaultValue=<rgba 12 32 12 1>}
  n36 = VectorParameter() {group=DirtAndEdgeWear, parameterName=DE_VehicleEdgeWearMaskTexture_Tiling, defaultValue=<rgba 12 32 12 1>}
  n37 = MaterialFunctionCall(
   TextureObject=StaticSwitchParameter(True=NamedRerouteUsage(), False=NamedRerouteUsage()) {group=DirtAndEdgeWear, parameterName="DE_EdgeWearMaskTexture------->USE GrungeTexture", sortPriority=10}
   TextureSize=LinearInterpolate(
    A=LinearInterpolate(A=@n35.RGB, B=@n36.RGB, Alpha=@n2)
    B=VectorParameter() {group=DirtAndEdgeWear, parameterName=DE_MechEdgeWearMaskTexture_Tiling, defaultValue=<rgba 12 32 12 1>}.RGB
    Alpha=@n5
   )
   WorldPosition=@n13
   Export Float 4=StaticBool() {value=true}
   World Space Normal=@n14
  ) {materialFunction=/Engine/Functions/Engine_MaterialFunctions01/Texturing/WorldAlignedTexture}
  n38 = ScalarParameter() {defaultValue=20, group=DirtAndEdgeWear, parameterName=DE_EdgeWearGlobalMaskStrength}
  n39 = ScalarParameter() {defaultValue=20, group=DirtAndEdgeWear, parameterName=DE_VehicleEdgeWearGlobalMaskStrength}
  n40 = Multiply(
   A=Power(Base=ChannelMaskParameter(None=@n37.XYZ Texture) {group=DirtAndEdgeWear, parameterName=DE_EdgeWearMaskTextureChannel}.R)
   B=LinearInterpolate(
    A=LinearInterpolate(A=@n38, B=@n39, Alpha=@n2)
    B=ScalarParameter() {defaultValue=20, group=DirtAndEdgeWear, parameterName=DE_MechEdgeWearGlobalMaskStrength}
    Alpha=@n5
   )
  )
  n41 = Saturate(None=OneMinus(None=Round(None=@n40)))
  n42 = Saturate(None=Multiply(
    A=SmoothStep(Max=@n32, Value=@n33)
    B=StaticSwitchParameter(True=@n34, False=@n41) {group=DirtAndEdgeWear, parameterName="DE_EdgeWearMask---->Just Crevase", sortPriority=10}
   ))
  n43 = StaticSwitchParameter(
   True=BlendMaterialAttributes(
    A=@n24.MaterialAttributes
    B=SetMaterialAttributes!(MaterialAttributes=@n29, Metallic=@n30, Specular=@n31) {attributeSetTypes [#4, #5]}
    Alpha=@n42
   ) {vertexAttributeBlendType=UseA}
   False=@n24.MaterialAttributes
  ) {defaultValue=true, group=DirtAndEdgeWear, parameterName=UseEdgeWear}
  n44 = LinearInterpolate(
   A=LinearInterpolate(
    A=VectorParameter() {group=DirtAndEdgeWear, parameterName=DE_WeaponDirtLayer_Tiling, defaultValue=<rgba 512 512 512 1>}.RGB
    B=VectorParameter() {group=DirtAndEdgeWear, parameterName=DE_VehicleDirtLayer_Tiling, defaultValue=<rgba 512 512 512 1>}.RGB
    Alpha=@n2
   )
   B=VectorParameter() {group=DirtAndEdgeWear, parameterName=DE_MechDirtLayer_Tiling, defaultValue=<rgba 512 512 512 1>}.RGB
   Alpha=@n5
  )
  n45 = Add(A=@n13, B=VectorParameter() {group=DirtAndEdgeWear, parameterName=DE_DirtLayer_Offset, defaultValue=<rgba 0 0 0 1>}.RGB)
  n46 = LinearInterpolate(
   A=VectorParameter() {group=DirtAndEdgeWear, parameterName=DE_WeaponDirtLayer_Rotation, defaultValue=<rgba 0 0 0 1>}.RGB
   B=VectorParameter() {group=DirtAndEdgeWear, parameterName=DE_VehicleDirtLayer_Rotation, defaultValue=<rgba 0 0 0 1>}.RGB
   Alpha=@n2
  )
  n47 = VectorParameter() {group=DirtAndEdgeWear, parameterName=DE_MechDirtLayer_Rotation, defaultValue=<rgba 0 0 0 1>}
  n48 = MaterialFunctionCall(Float3=Divide(A=LinearInterpolate(A=@n46, B=@n47.RGB, Alpha=@n5))) {materialFunction=&BreakOutFloat3Components}
  n49 = MaterialFunctionCall(
   Export Float 4=@n21
   TextureObject=NamedRerouteUsage()
   TextureSize=@n44
   WorldPosition=@n45
   World Space Normal=@n14
   RotZAxis=@n48.B
   RotXAxis=@n48.B
   RotYAxis=@n48.B
  ) {materialFunction=&/WorldAlignedTexture}
  n50 = StaticSwitchParameter(
   True=SetMaterialAttributes!(
    MaterialAttributes=@n24.MaterialAttributes
    BaseColor=VectorParameter() {group=DirtAndEdgeWear, parameterName=DirtLayerBaseColor, defaultValue=<rgba 0.0521 0.0329 0.0217 1>}.RGB
    Roughness=ScalarParameter() {defaultValue=1, group=DirtAndEdgeWear, parameterName=DE_DirLayerSimpleRoughness}
   ) {attributeSetTypes [#2, #3]}
   False=SetMaterialAttributes(
    MaterialAttributes=@n24.MaterialAttributes
    BaseColor=ComponentMask(None=@n49.XYZ Texture) {b=true, g=true, r=true}
    Roughness=ComponentMask(None=@n49.XYZ Texture) {a=true}
   ) {attributeSetTypes [#2, #3]}
  ) {defaultValue=true, group=DirtAndEdgeWear, parameterName=UseSimpleMaterialDirtLayer}
  n51 = ScalarParameter() {group=DirtAndEdgeWear, parameterName=DE_DirtLayer_Metalness}
  n52 = ScalarParameter() {group=DirtAndEdgeWear, parameterName=DE_DirtLayer_Specular}
  n53 = FunctionInput(Preview=@n1.A) {bUsePreviewValueAsDefault=true, inputName=GMap, inputType=FunctionInput_Scalar}
  n54 = ScalarParameter() {defaultValue=0.5, group=DirtAndEdgeWear, parameterName=DE_WeaponGMap_Spread}
  n55 = ScalarParameter() {defaultValue=0.5, group=DirtAndEdgeWear, parameterName=DE_VehicleGMap_Spread}
  n56 = Subtract(
   A=LinearInterpolate(A=@n53, B=OneMinus(None=@n53), Alpha=ScalarParameter() {group=DirtAndEdgeWear, parameterName=DE_FlipGMap})
   B=LinearInterpolate(
    A=LinearInterpolate(A=@n54, B=@n55, Alpha=@n2)
    B=ScalarParameter() {defaultValue=0.5, group=DirtAndEdgeWear, parameterName=DE_MechGMap_Spread}
    Alpha=@n5
   )
  )
  n57 = LinearInterpolate(
   A=LinearInterpolate(
    A=ScalarParameter() {defaultValue=2.5, group=DirtAndEdgeWear, parameterName=DE_WeaponGMap_Strength}
    B=ScalarParameter() {defaultValue=2.5, group=DirtAndEdgeWear, parameterName=DE_VehicleGMap_Strength}
    Alpha=@n2
   )
   B=ScalarParameter() {defaultValue=2.5, group=DirtAndEdgeWear, parameterName=DE_MechGMap_Strength}
   Alpha=@n5
  )
  n58 = VectorParameter() {group=DirtAndEdgeWear, parameterName=DE_WeaponDirtGrungeTexture_Tiling, defaultValue=<rgba 32 32 32 1>}
  n59 = VectorParameter() {group=DirtAndEdgeWear, parameterName=DE_VehicleDirtGrungeTexture_Tiling, defaultValue=<rgba 32 32 32 1>}
  n60 = MaterialFunctionCall(
   TextureObject=NamedRerouteUsage()
   TextureSize=LinearInterpolate(
    A=LinearInterpolate(A=@n58.RGB, B=@n59.RGB, Alpha=@n2)
    B=VectorParameter() {group=DirtAndEdgeWear, parameterName=DE_MechDirtGrungeTexture_Tiling, defaultValue=<rgba 32 32 32 1>}.RGB
    Alpha=@n5
   )
   WorldPosition=@n13
   Export Float 4=StaticBool() {value=true}
   World Space Normal=@n14
  ) {materialFunction=/Engine/Functions/Engine_MaterialFunctions01/Texturing/WorldAlignedTexture}
  n61 = LinearInterpolate(
   A=ScalarParameter() {defaultValue=0.5, group=DirtAndEdgeWear, parameterName=DE_DirtGrungeNoiseStrength}
   B=ScalarParameter() {defaultValue=0.5, group=DirtAndEdgeWear, parameterName=DE_VehicleDirtGrungeNoiseStrength}
   Alpha=@n2
  )
  n62 = ScalarParameter() {defaultValue=0.5, group=DirtAndEdgeWear, parameterName=DE_MechDirtGrungeNoiseStrength}
  n63 = Subtract(
   A=Saturate(None=Multiply(A=@n56, B=@n57))
   B=Multiply(
    A=ChannelMaskParameter(None=@n60.XYZ Texture) {group=DirtAndEdgeWear, parameterName=DE_DirtGrungeTextureChannel}.R
    B=LinearInterpolate(A=@n61, B=@n62, Alpha=@n2)
   )
  )
  n64 = Saturate(None=@n63)
  n65 = Constant()
  n66 = StaticSwitchParameter(
   True=BlendMaterialAttributes(
    A=@n43
    B=SetMaterialAttributes!(MaterialAttributes=@n50, Metallic=@n51, Specular=@n52) {attributeSetTypes [#4, #5]}
    Alpha=LinearInterpolate(A=@n64, B=@n65, Alpha=@n42)
   ) {vertexAttributeBlendType=UseA}
   False=@n43
  ) {defaultValue=true, group=DirtAndEdgeWear, parameterName=UseDirtLayer}
  n67 = FunctionInput() {
   bUsePreviewValueAsDefault=true
   desc=secondary
   inputName="Customization Mask 2 (_CM B Top Right)"
   inputType=FunctionInput_Scalar
   sortPriority=25
  }
  n68 = FunctionInput() {
   bUsePreviewValueAsDefault=true
   desc=base customization
   inputName="Customization Mask (_CM B Top left)"
   inputType=FunctionInput_Scalar
   sortPriority=24
  }
  n69 = Add(A=@n67, B=@n68) {desc="Base + Secondary = exposed primary"}
  n70 = OneMinus(None=@n69) {desc=Primary}
  n71 = StaticSwitchParameter(True=@n67, False=@n68) {desc=False uses Base customization, group=DirtAndEdgeWear, parameterName=DE_OnlySecondary, sortPriority=-99}
  n72 = StaticSwitchParameter(
   True=Add(A=@n70, B=@n68)
   False=StaticSwitchParameter(True=@n69, False=StaticSwitchParameter(True=@n70, False=@n71) {group=DirtAndEdgeWear, parameterName=DE_OnlyPrimary, sortPriority=-98}) {group=DirtAndEdgeWear, parameterName=DE_BaseAndSecondary, sortPriority=-97}
  ) {group=DirtAndEdgeWear, parameterName=DE_BaseAndPrimary, sortPriority=-96}
  n73 = Add(A=@n70, B=@n67)
  n74 = Saturate(None=StaticSwitchParameter(
    True=Constant() {r=1}
    False=StaticSwitchParameter(True=@n73, False=@n72) {group=DirtAndEdgeWear, parameterName=DE_SecondaryAndPrimary, sortPriority=-95}
   ) {defaultValue=true, group=DirtAndEdgeWear, parameterName=DE_CoverEntireObject, sortPriority=-100})
  FunctionOutput(None=StaticSwitchParameter(True=BlendMaterialAttributes(A=@n24.MaterialAttributes, B=@n66, Alpha=@n74) {vertexAttributeBlendType=UseA}, False=@n24.MaterialAttributes) {group=Customization, parameterName=Effect_DirtAndEdgeWearMasked})
  MaterialFunctionCall(TextureObject=NamedRerouteUsage(), TextureSize=@n44, WorldPosition=@n45, Export Float 4=@n21, World Space Normal=@n14) {materialFunction=&WorldAlignedTexture}
  n75 = ChannelMaskParameter(None=MaterialFunctionCall(
    TextureObject=NamedRerouteUsage()
    TextureSize=VectorParameter() {group=DirtAndEdgeWear, parameterName=DE_DirtMaskTexture_Tiling, defaultValue=<rgba 512 512 512 1>}.RGB
    WorldPosition=@n13
    Export Float 4=StaticBool() {value=true}
    World Space Normal=@n14
   ) {materialFunction=&WorldAlignedTexture}.XYZ Texture) {group=DirtAndEdgeWear, parameterName=DE_DirtMaskTextureChannel}
  n76 = OneMinus(None=Saturate(None=Power(Base=@n75.R)))
  Multiply(A=@n63, B=@n76)
  Multiply(A=@n76, B=ScalarParameter() {defaultValue=50, group=DirtAndEdgeWear, parameterName=DE_DirtGlobalMaskStrength})
  NamedRerouteDeclaration(None=TextureObjectParameter() {group=DirtAndEdgeWear, parameterName=DE_BottomLayer_BaseColor}) {name=Tex_BottomLayer_BC, nodeColor=<rgba 1 0 0.9375 1>}
  n77 = TextureObjectParameter() {
   group=DirtAndEdgeWear
   parameterName=DE_BottomLayer_Normal
   samplerType=SAMPLERTYPE_Normal
   texture=&FlatNormalBlue
  }
  NamedRerouteDeclaration(None=@n77) {name=Tex_BottomLayer_N, nodeColor=<rgba 1 0 0.9375 1>}
  n78 = TextureObjectParameter() {
   group=DirtAndEdgeWear
   parameterName=DE__Tex_Shared__Msk
   samplerType=SAMPLERTYPE_Masks
   sortPriority=0
   texture=/Game/Global/Textures/Default/Noise/T_WrapEdgeWear_Noise
  }
  n79 = TextureObjectParameter() {
   group=DirtAndEdgeWear
   parameterName=DE_DirtMaskTexture
   samplerType=SAMPLERTYPE_Masks
   texture=/Game/Global/Textures/Default/Noise/T_Noise_Cloud_Mask
  }
  n80 = StaticBoolParameter() {group=DirtAndEdgeWear, parameterName=DE__Tex_Shared_Msk, sortPriority=0}
  NamedRerouteDeclaration(None=StaticSwitch(True=@n78, False=@n79, Value=@n80)) {name=Tex_DirtMsk, nodeColor=<rgba 1 0 0.9375 1>}
  n81 = TextureObjectParameter() {
   group=DirtAndEdgeWear
   parameterName=DE_DirtGrungeTexture
   samplerType=SAMPLERTYPE_Masks
   texture=&T_Noise_Cloud_Mask
  }
  NamedRerouteDeclaration(None=StaticSwitch(True=@n78, False=@n81, Value=@n80)) {name=Tex_DirtGrundge, nodeColor=<rgba 1 0 0.9375 1>}
  n82 = TextureObjectParameter() {
   group=DirtAndEdgeWear
   parameterName=DE_EdgeWearMaskTexture
   samplerType=SAMPLERTYPE_Masks
   texture=&T_WrapEdgeWear_Noise
  }
  NamedRerouteDeclaration(None=StaticSwitch(True=@n78, False=@n82, Value=@n80)) {name=Tex_EdgeWearMsk, nodeColor=<rgba 1 0 0.9375 1>}
  n83 = TextureObjectParameter() {
   group=DirtAndEdgeWear
   parameterName=DE_EdgeWearGrungeTexture
   samplerType=SAMPLERTYPE_Masks
   texture=/Game/Global/Textures/Default/Noise/T_WrapEdgeWear_Noise_02
  }
  NamedRerouteDeclaration(None=StaticSwitch(True=@n78, False=@n83, Value=@n80)) {name=Tex_EdgeGrunge, nodeColor=<rgba 1 0 0.9375 1>}
  NamedRerouteDeclaration(None=@n24.Normal) {name=Mat_Normals, nodeColor=<rgba 0 1 0.1797 1>}
  NamedRerouteDeclaration(None=GetMaterialAttributes(None=FunctionInput() {bUsePreviewValueAsDefault=true, inputName=Root_Material, inputType=FunctionInput_MaterialAttributes}) {attributeGetTypes [#3, #1]}.Normal) {name=Mat_Normals_Root, nodeColor=<rgba 0 1 0.1797 1>}
  Saturate(None=@n8)
  TextureObjectParameter() {group=DirtAndEdgeWear, parameterName=DE_BottomLayer_BaseColor}
  TextureObjectParameter() {
   group=DirtAndEdgeWear
   parameterName=DE_BottomLayer_Normal
   samplerType=SAMPLERTYPE_Normal
   texture=&FlatNormalBlue
  }
 ]}
