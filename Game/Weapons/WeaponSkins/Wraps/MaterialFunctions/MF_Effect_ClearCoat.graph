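# Weapon wrap material function that adds an optional clear coat effect to incoming material attributes, gated by the Effect_ClearCoat static switch.
# It builds a clear coat variant with scaled metallic, using one parameter to swap map and constant roughness between base and coat.
# A mask from CM primary and accent channels, optional TopSide projection layers and decal overrides drives the attribute blend.
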MaterialFunction /Game/Weapons/WeaponSkins/Wraps/MaterialFunctions/MF_Effect_ClearCoat
graph {nodes [
  Constant2Vector()
  n1 = NamedRerouteUsage()
  n2 = NamedRerouteUsage()
  n3 = FunctionInput() {inputName=Decal_Mask, inputType=FunctionInput_Scalar, sortPriority=90}
  n4 = Constant()
  n5 = StaticSwitchParameter(
   True=ChannelMaskParameter(None=FunctionInput() {inputName=Decal_Mask_RGBA, inputType=FunctionInput_Vector4, sortPriority=90}) {group=Effect-ClearCoat_Msk, parameterName=ECCoat_Decal_Mask_RGBA_Channel}.R
   False=@n4
  ) {group=Effect-ClearCoat_Msk, parameterName=ECCoat_Decal_Mask_RGBA}
  n6 = OneMinus(None=LinearInterpolate(B=@n3, Alpha=ScalarParameter() {defaultValue=1, group=Effect-ClearCoat_Msk, parameterName="ECCoat_Decal_Alpha Masked  1=Active"}))
  n7 = Constant() {r=1}
  n8 = FunctionInput() {inputName=TS_Layer02_Mask, inputType=FunctionInput_Vector4, sortPriority=99}
  n9 = ChannelMaskParameter(None=FunctionInput() {inputName=TS_Layer03_Mask, inputType=FunctionInput_Vector4, sortPriority=99}) {group=Effect-ClearCoat_Msk, parameterName=ECCoat_TS_Layer03_Mask_Channel}
  n10 = Constant() {r=1}
  n11 = StaticSwitchParameter(
   True=ChannelMaskParameter(None=FunctionInput() {inputName=TS_Layer04_Mask, inputType=FunctionInput_Vector4, sortPriority=99}) {group=Effect-ClearCoat_Msk, parameterName=ECCoat_TS_Layer04_Mask_Channel}.R
   False=@n10
  ) {group=Effect-ClearCoat_Msk, parameterName=ECCoat_TS_Layer04_Mask}
  n12 = StaticSwitchParameter(
   True=ChannelMaskParameter(None=FunctionInput() {inputName=TS_Layer01_Mask, inputType=FunctionInput_Vector4, sortPriority=99}) {group=Effect-ClearCoat_Msk, parameterName=ECCoat_TS_Layer01_Mask_Channel}.R
   False=StaticSwitchParameter(
    True=ChannelMaskParameter(None=@n8) {group=Effect-ClearCoat_Msk, parameterName=ECCoat_TS_Layer02_Mask_Channel}.R
    False=StaticSwitchParameter(True=@n9.R, False=@n11) {group=Effect-ClearCoat_Msk, parameterName=ECCoat_TS_Layer03_Mask}
   ) {group=Effect-ClearCoat_Msk, parameterName=ECCoat_TS_Layer02_Mask}
  ) {group=Effect-ClearCoat_Msk, parameterName=ECCoat_TS_Layer01_Mask}
  n13 = StaticBoolParameter() {group=TopSide_V02, parameterName=TopSideProjection_V02}
  n14 = ScalarParameter() {defaultValue=1, group=Effect-ClearCoat, parameterName="ECCoat_Primary   0=Off"}
  n15 = Constant()
  n16 = LinearInterpolate(A=@n15, B=NamedRerouteUsage(), Alpha=ScalarParameter() {group=TopSide_V02, parameterName="TopS_Mask_Accent  0=Off"})
  n17 = ScalarParameter() {group=TopSide_V02, parameterName="TopS_Mask_Primary\\Accent  1=Off"}
  n18 = TextureSample() {samplerType=SAMPLERTYPE_Masks, texture=/Game/Weapons/FORT_Rifles/Textures/T_Scar_CM}
  n19 = MaterialFunctionCall(Float4=FunctionInput(Preview=@n18.RGBA) {inputName=CM_Msk, inputType=FunctionInput_Vector4, sortPriority=4}) {materialFunction=&BreakOutFloat4Components}
  n20 = Max(
   A=StaticSwitch(True=Max(A=@n16, B=@n17), False=@n15, Value=StaticBoolParameter() {group=TopSide_V02, parameterName=TopSideProjection_V02})
   B=NamedRerouteDeclaration(None=@n19.R) {name=Msk_Prime, nodeColor=<rgba 0 1 0.7422 1>}
  )
  n21 = Max(
   A=Min(A=StaticSwitch(True=@n12, False=@n10, Value=@n13), B=Min(A=@n14, B=@n20))
   B=Min(
    A=NamedRerouteDeclaration(None=@n19.B) {name=Msk_Accent, nodeColor=<rgba 0 1 0.7656 1>}
    B=ScalarParameter() {group=Effect-ClearCoat, parameterName="ECCoat_Accent    0=Off"}
   )
  )
  n22 = ScalarParameter() {group=Effect-ClearCoat, parameterName="ECCoat_Mask_Disable 1=Active", sortPriority=0}
  n23 = Max(
   A=StaticSwitchParameter(True=StaticSwitchParameter(True=@n3, False=@n5) {group=Effect-ClearCoat_Msk, parameterName=ECCoat_Decal_Mask_Alpha}, False=@n4) {group=Customization, parameterName=Function_Left_Right_Decal_V2}
   B=Min(A=StaticSwitchParameter(True=@n6, False=@n7) {group=Customization, parameterName=Function_Left_Right_Decal_V2}, B=Max(A=@n21, B=@n22))
  )
  n24 = TextureSample() {samplerType=SAMPLERTYPE_Normal, texture=/Game/Weapons/FORT_Rifles/T_SCAR_N}
  n25 = Constant3Vector() {constant=<rgba 1 0 0 0>}
  n26 = GetMaterialAttributes(None=FunctionInput(Preview=SetMaterialAttributes!(MaterialAttributes, Normal=@n24.RGB, BaseColor=@n25) {attributeSetTypes [#1, #2]}) {inputName=Material, inputType=FunctionInput_MaterialAttributes}) {attributeGetTypes [
    #2
    #3
    #1
    #4
   ]}
  FunctionOutput(None=StaticSwitchParameter(
    True=BlendMaterialAttributes(A=@n1, B=@n2, Alpha=@n23) {vertexAttributeBlendType=UseA}
    False=NamedRerouteDeclaration(None=@n26.MaterialAttributes) {name=Root_Material, nodeColor=<rgba 0.7812 1 0 1>}
   ) {group=Customization, parameterName=Effect_ClearCoat}) {outputName=Result_1}
  n27 = GetMaterialAttributes(None=NamedRerouteUsage()) {attributeGetTypes [#3, #4]}
  n28 = ScalarParameter() {defaultValue=0.1, group=Effect-ClearCoat, parameterName=ECCoat_Undercoat Roughness, sortPriority=100}
  n29 = ScalarParameter() {group=Effect-ClearCoat, parameterName=ECCoat_Roughness Map Now affects Clearcoat roughness, sortPriority=100}
  n30 = SetMaterialAttributes!(
   MaterialAttributes=@n27.MaterialAttributes
   CustomData0=Constant() {r=1}
   CustomData1=LinearInterpolate(A=@n28, B=@n27.Roughness, Alpha=@n29)
   Roughness=LinearInterpolate(A=@n27.Roughness, B=@n28, Alpha=@n29)
   Metallic=Multiply(
    A=@n27.Metallic
    B=ScalarParameter() {defaultValue=1, group=Effect-ClearCoat, parameterName=ECCoat_Undercoat Metallic Multiplier, sortPriority=100}
   )
   ShadingModel=ShadingModel() {shadingModel=MSM_ClearCoat}
  ) {attributeSetTypes [
    #5
    #6
    #3
    #4
    #7
   ]}
  NamedRerouteDeclaration(None=@n30) {name=Material, nodeColor=<rgba 0 1 0.9766 1>}
  NamedRerouteDeclaration(None=@n19.G) {name=Msk_Curvature, nodeColor=<rgba 0 1 0.7422 1>}
  NamedRerouteDeclaration(None=@n18.RGBA) {name=CM_Mask, nodeColor=<rgba 1 0 0.3047 1>}
  n31 = Constant4Vector()
  NamedRerouteDeclaration(None=FunctionInput(Preview=@n31.RGBA) {bUsePreviewValueAsDefault=true, inputName=IsState, inputType=FunctionInput_Vector4, sortPriority=4}) {name=IsState, nodeColor=<rgba 0.1497 1 0.1633 1>}
  NamedRerouteDeclaration(None=FunctionInput(Preview=@n31.RGBA) {bUsePreviewValueAsDefault=true, inputName=IsState_B, inputType=FunctionInput_Vector4, sortPriority=4}) {name=IsState_B, nodeColor=<rgba 0.1497 1 0.1633 1>}
  n32 = PreSkinnedLocalBounds()
  n33 = Subtract(A=VertexInterpolator(VS=PreSkinnedPosition()).PS, B=@n32.Min)
  n34 = MaterialFunctionCall(Color=Subtract(A=@n32.Max, B=@n32.Max)) {materialFunction=&SplitComponents}
  n35 = Max(A=Max(A=@n34.R, B=@n34.R), B=@n34.B)
  NamedRerouteDeclaration(None=FunctionInput(Preview=Divide(A=@n33, B=@n35)) {bUsePreviewValueAsDefault=true, inputName=Calcated_Bounds, sortPriority=11}) {name=Caclcated_Bounds, nodeColor=<rgba 0 0.9297 1 1>}
  n36 = PreSkinnedNormal()
  NamedRerouteDeclaration(None=FunctionInput(Preview=VertexInterpolator(VS=@n36).PS) {bUsePreviewValueAsDefault=true, inputName=Pre-Skinned Local Normal, sortPriority=10}) {name=PreSkin_Normal, nodeColor=<rgba 0.8281 1 0 1>}
  NamedRerouteDeclaration(None=@n26.Normal) {name=Root_Normal, nodeColor=<rgba 0.4062 0 1 1>}
 ]}
