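# Weapon wrap material function that adds an optional cloth fuzz effect to an incoming material, toggled by a static switch.
# The new shading path uses the Cloth shading model with tinted subsurface, biased roughness and Fresnel-scaled specular as a Substrate fix, reducing to tinted base color on mobile.
# The older path tints base color through a tiled fuzz-texture Fresnel, skipped on mobile.
# Primary, secondary, channel-selected and decal masks build the blend alpha.
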MaterialFunction /Game/Weapons/WeaponSkins/Wraps/MaterialFunctions/MF_Effect_ClothFuzz
graph {nodes [
  n1 = NamedRerouteUsage()
  n2 = GetMaterialAttributes(None=FunctionInput() {inputName=Material, inputType=FunctionInput_MaterialAttributes}) {attributeGetTypes [
    #1
    #2
    #3
    #4
   ]}
  n3 = Multiply(A=@n2.BaseColor, B=ScalarParameter() {defaultValue=1, group=Effect_ClothFuzz, parameterName=CF_Cloth Subsurface Intensity})
  n4 = ScalarParameter() {defaultValue=0.4, group=06_Cloth_SubstrateFix, parameterName=SubstrateClothFix_RoughnessBias, sliderMax=1, sortPriority=0}
  n5 = Saturate(None=Multiply(A=@n2.Roughness, B=ScalarParameter() {defaultValue=1, group=Effect_ClothFuzz, parameterName=CF_Cloth Roughness}))
  n6 = LinearInterpolate(
   A=ScalarParameter() {defaultValue=0.75, group=Effect_ClothFuzz, parameterName=CF_Cloth Scattering_Spec}
   B=Multiply(A=ScalarParameter() {defaultValue=1, group=Effect_ClothFuzz, parameterName=CF_Cloth Scattering_Spec_Root}, B=NamedRerouteUsage())
   Alpha=ScalarParameter() {group=Effect_ClothFuzz, parameterName="CF_Cloth Scattering_Spec_Root_influnce   1=active"}
  )
  n7 = ScalarParameter() {defaultValue=5, group=06_Cloth_SubstrateFix, parameterName=SubstrateClothFix_SpecularFresnelExponent, sortPriority=0}
  n8 = ScalarParameter() {defaultValue=0.3, group=06_Cloth_SubstrateFix, parameterName=SubstrateClothFix_SpecularFresnelBase, sortPriority=0}
  n9 = SetMaterialAttributes!(
   MaterialAttributes=NamedRerouteUsage()
   ShadingModel=ShadingModel() {shadingModel=MSM_Cloth}
   SubsurfaceColor=Saturate(None=Multiply(A=@n1, B=@n3))
   Roughness=LinearInterpolate(A=@n4, Alpha=Saturate(None=@n5))
   CustomData0=ScalarParameter() {defaultValue=1, group=Effect_ClothFuzz, parameterName=CF_Cloth Shader Intensity}
   Opacity=Constant() {r=1}
   BaseColor
   Specular=Multiply(A=Saturate(None=@n6), B=Fresnel(ExponentIn=@n7, BaseReflectFractionIn=@n8) {baseReflectFraction=0})
   Normal=@n2.Normal
  ) {attributeSetTypes [
    #5
    #6
    #3
    #7
    #8
    #1
    #4
    #2
   ]}
  n10 = NamedRerouteUsage()
  n11 = Saturate(None=Multiply(A=@n2.BaseColor, B=NamedRerouteUsage()))
  n12 = ShadingPathSwitch(
   Default=BlendMaterialAttributes(A=NamedRerouteUsage(), B=@n9, Alpha=NamedRerouteUsage()) {vertexAttributeBlendType=UseA}
   Mobile=BlendMaterialAttributes(
    A=NamedRerouteUsage()
    B=SetMaterialAttributes!(MaterialAttributes=@n10, BaseColor=@n11) {attributeSetTypes [#1]}
    Alpha=NamedRerouteUsage()
   ) {vertexAttributeBlendType=UseA}
  )
  n13 = DotProduct(A=PixelNormalWS(), B=CameraVectorWS())
  n14 = Saturate(None=Power(Base=OneMinus(None=@n13), Exp=ScalarParameter() {defaultValue=4, group=Effect_ClothFuzz, parameterName=CF_Fuzz Exponent}))
  n15 = TextureSampleParameter2D(UVs=Multiply(A=TextureCoordinate(), B=ScalarParameter() {defaultValue=4, group=Effect_ClothFuzz, parameterName=CF_Fuzz Tiling})) {
   group=Effect_ClothFuzz
   parameterName=CF_ClothFuzz Texture
   samplerSource=SSM_Wrap_WorldGroupSettings
   samplerType=SAMPLERTYPE_LinearColor
   texture=/Game/Global/Textures/Default/Masks_and_Gradients/T_Fuzz_MASK
  }
  n16 = Saturate(None=Multiply(A=Multiply(A=@n14, B=@n15.R), B=ScalarParameter() {defaultValue=1, group=Effect_ClothFuzz, parameterName=CF_Fuzz Fresnel Blend}))
  n17 = LinearInterpolate(
   A=@n2.BaseColor
   B=Multiply(A=@n2.BaseColor, B=VectorParameter() {group=Effect_ClothFuzz, parameterName=CF_Fuzz Tint, defaultValue=<rgba 0.6823 0.6823 0.6823 1>}.RGB)
   Alpha=@n16
  )
  n18 = ShadingPathSwitch(
   Default=BlendMaterialAttributes(
    A=NamedRerouteUsage()
    B=SetMaterialAttributes(MaterialAttributes=@n2.MaterialAttributes, BaseColor=@n17) {attributeSetTypes [#1]}
    Alpha=NamedRerouteUsage()
   ) {vertexAttributeBlendType=UseA}
   Mobile=NamedRerouteUsage()
  )
  FunctionOutput(None=StaticSwitchParameter(
    True=StaticSwitchParameter(True=@n12, False=@n18) {defaultValue=true, group=Effect_ClothFuzz, parameterName=CF_NewClothShading, sortPriority=-1}
    False=NamedRerouteUsage()
   ) {group=Customization, parameterName=Effect_ClothFuzz})
  NamedRerouteDeclaration(None=VectorParameter() {group=Effect_ClothFuzz, parameterName=CF_Cloth Fuzz Tint, defaultValue=<rgba 1 1 1 1>}.RGB) {name=Cloth Color, nodeColor=<rgba 0 1 0.2969 1>}
  NamedRerouteDeclaration(None=@n2.MaterialAttributes) {name=Root_Material, nodeColor=<rgba 0.5703 1 0 1>}
  n19 = FunctionInput(Preview=Constant()) {bUsePreviewValueAsDefault=true, inputName=Decal_Mask, inputType=FunctionInput_Scalar}
  n20 = OneMinus(None=@n19)
  n21 = FunctionInput() {bUsePreviewValueAsDefault=true, inputName=Mask_secondary, inputType=FunctionInput_Scalar, sortPriority=1}
  n22 = ScalarParameter() {group=Effect_ClothFuzz, parameterName="CF_MaskAccent 1=enable", sortPriority=0}
  n23 = Add(
   A=ScalarParameter() {defaultValue=1, group=Effect_ClothFuzz, parameterName="CF_Mask Primary 0=enable", sortPriority=0}
   B=FunctionInput() {bUsePreviewValueAsDefault=true, inputName=Mask_Primary, inputType=FunctionInput_Scalar, sortPriority=1}
  )
  n24 = StaticSwitchParameter(
   True=ChannelMaskParameter(None=FunctionInput() {bUsePreviewValueAsDefault=true, inputName=TS_Mask_02, inputType=FunctionInput_Vector4, sortPriority=3}) {group=Effect_ClothFuzz, parameterName=CF_TS_Mask_02_Channel, sortPriority=2}.R
   False=ChannelMaskParameter(None=FunctionInput() {bUsePreviewValueAsDefault=true, inputName=TS_Mask_01, inputType=FunctionInput_Vector4, sortPriority=2}) {group=Effect_ClothFuzz, parameterName=CF_TS_Mask_01_Channel, sortPriority=2}.R
  ) {group=Effect_ClothFuzz, parameterName="CF_Mask TS_Mask_01----->TS_Mask_02", sortPriority=1}
  n25 = Constant() {r=1}
  n26 = Add(
   A=Saturate(None=Multiply(A=@n21, B=@n22))
   B=Multiply(
    A=Saturate(None=@n23)
    B=StaticSwitchParameter(True=@n24, False=@n25) {group=Effect_ClothFuzz, parameterName=CF_Mask Use TS_Mask_01, sortPriority=1}
   )
  )
  n27 = Saturate(None=LinearInterpolate(
    A=Multiply(A=@n20, B=@n26)
    B=Add(A=@n19, B=@n26)
    Alpha=ScalarParameter() {defaultValue=1, group=Effect_ClothFuzz, parameterName="CF_Mask Left/Right_Decal 1=off 0=on", sortPriority=0}
   ))
  n28 = Saturate(None=@n26)
  NamedRerouteDeclaration(None=StaticSwitch(
    True=StaticSwitchParameter(True=@n27, False=@n28) {group=Customization, parameterName=Function_Left_Right_Decal_V2}
    False=StaticSwitchParameter(True=@n27, False=@n28) {group=Customization, parameterName=Function_Left_Right_Decal}
    Value=StaticBoolParameter() {group=Customization, parameterName="Custmization  V2", sortPriority=-500}
   )) {name=Alpha Source, nodeColor=<rgba 1 0.2109 0 1>}
  NamedRerouteDeclaration(None=@n2.Specular) {name=Spec, nodeColor=<rgba 1 0 0.3281 1>}
  n29 = ScalarParameter() {defaultValue=0.9, group=Effect_ClothFuzz, parameterName=CF_Cloth Base Color Intensity}
  n30 = NamedRerouteUsage()
  n31 = Add(A=Multiply(A=@n16) {constB=10})
  Saturate(None=Multiply(A=Multiply(A=@n2.BaseColor, B=@n29), B=Multiply(A=@n30, B=@n31)))
 ]}
