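# Wrap material function that layers curvature-driven edge wear and pit dirt over an incoming material, behind the Effect_EdgeWear_Dirt switch.
# Edge and pit masks combine CM curvature with projected grunge, optional top/side masks and, for pits, Gmap.
# IsState inputs blend scales and strengths per item type, such as weapon, vehicle or character.
# Edges can react to audio or timers; primary, accent, decal, HotWax and TopSide masks limit the result.
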MaterialFunction /Game/Weapons/WeaponSkins/Wraps/MaterialFunctions/MF_Effect_EdgeWear_Dirt_V2
graph {nodes [
  n1 = Multiply(
   A=Subtract(
    A=NamedRerouteUsage()
    B=ScalarParameter() {defaultValue=1, desc="1 = disable", group=Effect_EdgeWear_Dirt_V2_Pits, parameterName="Eff_Edge\\Dirt_Pits_Gmap_Position"}
   )
   B=NamedRerouteUsage()
  )
  n2 = ComponentMask(None=NamedRerouteUsage()) {b=true}
  n3 = MaterialFunctionCall(Float4=NamedRerouteUsage()) {materialFunction=&BreakOutFloat4Components}
  n4 = Multiply(A=Divide(A=Add(A=@n2, B=@n3.R), B=@n3.G), B=@n3.B)
  Add(A=@n1, B=@n4)
  BlendMaterialAttributes() {vertexAttributeBlendType=UseA}
  CollectionParameter() {collection=&FortniteMaterialParameters, parameterName=EventAudio_Amplitude}
  Constant()
  Constant()
  n5 = NamedRerouteUsage()
  n6 = GetMaterialAttributes(None=NamedRerouteUsage()) {attributeGetTypes [#1]}
  n7 = NamedRerouteUsage()
  n8 = ScalarParameter() {group=Effect_EdgeWear_Dirt_V2, parameterName="Eff_Edge\\Dirt_Root_Emissive    1=on", sortPriority=0}
  n9 = SetMaterialAttributes!(MaterialAttributes=@n6.MaterialAttributes, EmissiveColor=Add(A=@n6.EmissiveColor, B=LinearInterpolate(B=@n7, Alpha=@n8))) {attributeSetTypes [#1]}
  n10 = OneMinus(None=FunctionInput() {inputName=HotWax_Msk, inputType=FunctionInput_Scalar})
  n11 = ScalarParameter() {defaultValue=1, group=Effect_EdgeWear_Dirt_V2, parameterName="Eff_Edge\\Dirt_HotWax_Msk 0=On", sortPriority=0}
  n12 = FunctionInput(Preview=Constant() {r=1}) {bUsePreviewValueAsDefault=true, inputName="Msk_Ts01<--->Ts02", inputType=FunctionInput_Scalar}
  n13 = LinearInterpolate(
   A=@n12
   B=OneMinus(None=@n12)
   Alpha=ScalarParameter() {group=Effect_EdgeWear_Dirt_V2, parameterName="Eff_Edge\\Dirt_Blend TopSide01---> 02 Invert Msk  1=inverted"}
  )
  n14 = Max(
   A=ScalarParameter() {defaultValue=1, group=Effect_EdgeWear_Dirt_V2, parameterName="Eff_Edge\\Dirt_Decal_Msk   1=On", sortPriority=0}
   B=OneMinus(None=FunctionInput() {inputName=Decal_Mask, inputType=FunctionInput_Scalar, sortPriority=4})
  )
  n15 = ScalarParameter() {defaultValue=1, group=Effect_EdgeWear_Dirt_V2, parameterName="Eff_Edge\\Dirt_Primary_Msk    1=On", sortPriority=0}
  n16 = TextureSample() {samplerType=SAMPLERTYPE_Masks, texture=/Game/Weapons/FORT_Rifles/Textures/T_Scar_CM}
  n17 = MaterialFunctionCall(Float4=FunctionInput(Preview=@n16.RGBA) {inputName=CM_Msk, inputType=FunctionInput_Vector4, sortPriority=4}) {materialFunction=&BreakOutFloat4Components}
  n18 = NamedRerouteDeclaration(None=@n17.R) {name=Msk_Prime, nodeColor=<rgba 0 1 0.7422 1>}
  n19 = NamedRerouteDeclaration(None=@n17.B) {name=Msk_Accent, nodeColor=<rgba 0 1 0.7656 1>}
  n20 = ScalarParameter() {group=Effect_EdgeWear_Dirt_V2, parameterName="Eff_Edge\\Dirt_Accent_Msk    1=On", sortPriority=0}
  n21 = Max(
   A=Max(A=Min(A=@n15, B=@n18), B=Min(A=@n19, B=@n20))
   B=ScalarParameter() {group=Effect_EdgeWear_Dirt_V2, parameterName="Eff_Edge\\Dirt_Msk_Disable    1=On", sortPriority=0}
  )
  n22 = Saturate(None=StaticSwitchParameter(True=Min(A=@n14, B=@n21), False=@n21) {group=Customization, parameterName=Function_Left_Right_Decal_V2})
  n23 = StaticSwitch(
   True=StaticSwitch(
    True=Min(A=@n13, B=@n22)
    False=@n22
    Value=StaticBoolParameter() {group=Effect_EdgeWear_Dirt_V2, parameterName="Eff_Edge\\Dirt_Blend TopSide01---> 02 --BlendMask"}
   )
   False=@n22
   Value=StaticBoolParameter() {group="TopSide_V02_Layer 01 Blend  02", parameterName="TopS Blend 01<--->02"}
  )
  n24 = StaticSwitchParameter(True=Min(A=Max(A=@n10, B=@n11), B=@n23), False=@n23) {group=Customization, parameterName=Effect_HotWax}
  n25 = TextureSample() {samplerType=SAMPLERTYPE_Normal, texture=/Game/Weapons/FORT_Rifles/T_SCAR_N}
  n26 = Constant3Vector() {constant=<rgba 1 0 0 0>}
  n27 = GetMaterialAttributes(None=FunctionInput(Preview=SetMaterialAttributes!(MaterialAttributes, Normal=@n25.RGB, BaseColor=@n26) {attributeSetTypes [#2, #3]}) {inputName=Material In, inputType=FunctionInput_MaterialAttributes}) {attributeGetTypes [
    #3
    #4
    #2
    #5
    #1
   ]}
  FunctionOutput(None=StaticSwitchParameter(
    True=BlendMaterialAttributes(A=@n5, B=@n9, Alpha=@n24) {vertexAttributeBlendType=UseA}
    False=NamedRerouteDeclaration(None=@n27.MaterialAttributes) {name=In_Material, nodeColor=<rgba 0.7812 1 0 1>}
   ) {group=Customization, parameterName=Effect_EdgeWear_Dirt}) {outputName=Result_1}
  n28 = GetMaterialAttributes(None=FunctionInput() {inputName=Material Root, inputType=FunctionInput_MaterialAttributes}) {attributeGetTypes [
    #3
    #4
    #2
    #5
   ]}
  NamedRerouteDeclaration(None=@n28.Normal) {name=Root_Normal, nodeColor=<rgba 0.4062 0 1 1>}
  NamedRerouteDeclaration(None=FunctionInput() {bUsePreviewValueAsDefault=true, inputName=TopsSide_Msk 01, inputType=FunctionInput_Vector4, sortPriority=15}) {name=TopSide_Msk01, nodeColor=<rgba 1 0 0.9609 1>}
  n29 = ScalarParameter() {defaultValue=5, group=Effect_EdgeWear_Dirt_V2_UV_Edge, parameterName="Eff_Edge\\Dirt_Edge_Grunge_Scale_Vehicle"}
  n30 = ScalarParameter() {defaultValue=6, group=Effect_EdgeWear_Dirt_V2_UV_Edge, parameterName="Eff_Edge\\Dirt_Edge_Grunge_Vehicle_Side"}
  n31 = MaterialFunctionCall(Float4=NamedRerouteUsage()) {materialFunction=&BreakOutFloat4Components}
  n32 = LinearInterpolate(
   A=LinearInterpolate(
    A=ScalarParameter() {defaultValue=5, group=Effect_EdgeWear_Dirt_V2_UV_Edge, parameterName="Eff_Edge\\Dirt_Edge_Grunge_Scale_Weapon"}
    B=LinearInterpolate(A=@n29, B=@n30, Alpha=@n31.G)
    Alpha=@n31.R
   )
   B=LinearInterpolate(
    A=ScalarParameter() {defaultValue=6, group=Effect_EdgeWear_Dirt_V2_UV_Edge, parameterName="Eff_Edge\\Dirt_Edge_Grunge_Scale_Mech"}
    B=ScalarParameter() {defaultValue=2, group=Effect_EdgeWear_Dirt_V2_UV_Edge, parameterName="Eff_Edge\\Dirt_Edge_Grunge_Scale_Character"}
    Alpha=@n31.A
   )
   Alpha=@n31.B
  )
  n33 = ScalarParameter() {defaultValue=2, group=Effect_EdgeWear_Dirt_V2_UV_Edge, parameterName="Eff_Edge\\Dirt_Edge_Grunge_Scale_Backpack"}
  n34 = MaterialFunctionCall(Float2=NamedRerouteUsage()) {materialFunction=&BreakOutFloat2Components}
  NamedRerouteDeclaration(None=LinearInterpolate(
    A=LinearInterpolate(A=@n32, B=@n33, Alpha=@n34.R)
    B=ScalarParameter() {defaultValue=2, group=Effect_EdgeWear_Dirt_V2_UV_Edge, parameterName="Eff_Edge\\Dirt_Edge_Grunge_Scale_Pickaxe"}
    Alpha=@n34.G
   )) {name=UVs_Scale_Edge_Grunge, nodeColor=<rgba 0 0.7188 1 1>}
  NamedRerouteDeclaration(None=FunctionInput() {bUsePreviewValueAsDefault=true, inputName=TopsSide_Msk 02, inputType=FunctionInput_Vector4, sortPriority=15}) {name=TopSide_Msk02, nodeColor=<rgba 1 0 0.9609 1>}
  n35 = StaticSwitchParameter(True=NamedRerouteUsage(), False=NamedRerouteUsage()) {group=Effect_EdgeWear_Dirt_V2_Edge, parameterName="Eff_Edge\\Dirt_Edge_TopSide_Msks01------->TopSide_Msk02"}
  n36 = Multiply(
   A=Subtract(
    A=ChannelMaskParameter(None=@n35) {group=Effect_EdgeWear_Dirt_V2_Edge, parameterName="Eff_Edge\\Dirt_Edge_TopSide_Msks_Channel"}.R
    B=ScalarParameter() {group=Effect_EdgeWear_Dirt_V2_Edge, parameterName="Eff_Edge\\Dirt_Edge_TopSide_Msks_Position"}
   )
   B=ScalarParameter() {defaultValue=1, group=Effect_EdgeWear_Dirt_V2_Edge, parameterName="Eff_Edge\\Dirt_Edge_TopSide_Msks_Strength"}
  )
  n37 = ScalarParameter() {defaultValue=1, group=Effect_EdgeWear_Dirt_V2_Edge, parameterName="Eff_Edge\\Dirt_Edge_TopSide_Msks_Fallof"}
  NamedRerouteDeclaration(None=Saturate(None=Power(Base=@n36, Exp=@n37))) {name=TopsSide_Edge, nodeColor=<rgba 1 0 0.7734 1>}
  n38 = ChannelMaskParameter(None=StaticSwitchParameter(True=NamedRerouteUsage(), False=NamedRerouteUsage()) {group=Effect_EdgeWear_Dirt_V2_Pits, parameterName="Eff_Edge\\Dirt_Pits_TopSide_Msks01------->TopSide_Msk02", sortPriority=50}) {group=Effect_EdgeWear_Dirt_V2_Pits, parameterName="Eff_Edge\\Dirt_Pits_TopSide_Msks_Channel", sortPriority=50}
  n39 = ScalarParameter() {defaultValue=0.5, group=Effect_EdgeWear_Dirt_V2_Pits, parameterName="Eff_Edge\\Dirt_Pits_TopSide_Msks_Position", sortPriority=50}
  n40 = ScalarParameter() {defaultValue=1, group=Effect_EdgeWear_Dirt_V2_Pits, parameterName="Eff_Edge\\Dirt_Pits_TopSide_Msks_Strength", sortPriority=50}
  n41 = Multiply(A=Subtract(A=OneMinus(None=@n38.R), B=@n39), B=@n40)
  n42 = ScalarParameter() {defaultValue=1, group=Effect_EdgeWear_Dirt_V2_Pits, parameterName="Eff_Edge\\Dirt_Pits_TopSide_Msks_Fallof", sortPriority=50}
  NamedRerouteDeclaration(None=Saturate(None=Power(Base=@n41, Exp=@n42))) {name=TopsSide_Pits, nodeColor=<rgba 1 0 0.7734 1>}
  n43 = CollectionParameter() {collection=&FortniteMaterialParameters, parameterName=EventAudio_Spectrum}
  n44 = ChannelMaskParameter(None=@n43) {group=Effect_EdgeWear_Dirt_V2_Edge_Reactive, maskChannel=Alpha, parameterName="Eff_Edge\\Dirt_Audio Channel", defaultValue=<rgba 0 0 0 1>}
  n45 = CollectionParameter() {collection=&FortniteMaterialParameters, parameterName=Emote_AudioLow}
  n46 = Max(A=@n44.R, B=@n45)
  n47 = ScalarParameter() {defaultValue=0.25, group=Effect_EdgeWear_Dirt_V2_Edge_Reactive, parameterName="Eff_Edge\\Dirt_Audio Low_Remap"}
  n48 = ScalarParameter() {defaultValue=1, group=Effect_EdgeWear_Dirt_V2_Edge_Reactive, parameterName="Eff_Edge\\Dirt_Audio High_Remap"}
  n49 = Constant()
  n50 = Constant() {r=1}
  n51 = CollectionParameter() {collection=&FortniteMaterialParameters, parameterName=Emote_AudioHigh}
  n52 = Max(
   A=Saturate(None=MaterialFunctionCall(Input=@n46, Input Low=@n47, Input High=@n48, Target Low=@n49, Target High=@n50) {materialFunction=&RemapValueRange}.Result)
   B=@n51
  )
  NamedRerouteDeclaration(None=@n52) {name=AudioSpectrum A, nodeColor=<rgba 1 0.8203 0 1>}
  NamedRerouteDeclaration(None=StaticBoolParameter() {group=Effect_EdgeWear_Dirt_V2_Edge_Reactive, parameterName="Eff_Edge\\Dirt_Audio Reactive Warp"}) {name=Audio_Warp, nodeColor=<rgba 0 0.4141 1 1>}
  n53 = VectorParameter() {group=Effect_EdgeWear_Dirt_V2_UV_Edge, parameterName="Eff_Edge\\Dirt_Edge_Offset", defaultValue=<rgba 0 0 0 1>}
  n54 = Time()
  n55 = VectorParameter() {group=Effect_EdgeWear_Dirt_V2_UV_Edge, parameterName="Eff_Edge\\Dirt_Edge_Offset_Speed", defaultValue=<rgba 0 0 0 1>}
  NamedRerouteDeclaration(None=Add(A=@n53.RGB, B=Multiply(A=@n54, B=@n55.RGB))) {name=UVs_Edge_Offset, nodeColor=<rgba 0 0.7188 1 1>}
  n56 = ScalarParameter() {defaultValue=1, group=Effect_EdgeWear_Dirt_V2_Edge_Reactive, parameterName="Eff_Edge\\Dirt_Audio Strength"}
  n57 = Time()
  n58 = ScalarParameter() {defaultValue=2, group=Effect_EdgeWear_Dirt_V2_Edge_Reactive, parameterName="Eff_Edge\\Dirt_Audio Static_Speed"}
  n59 = ConstantBiasScale(None=Sine(None=Multiply(A=@n57, B=@n58)))
  n60 = ScalarParameter() {group=Effect_EdgeWear_Dirt_V2_Edge_Reactive, parameterName="Eff_Edge\\Dirt_Audio Static_Strength"}
  NamedRerouteDeclaration(None=Max(A=Multiply(A=@n52, B=@n56), B=Multiply(A=@n59, B=@n60))) {name=AudioSpectrum A_Static_Speed, nodeColor=<rgba 1 0.8203 0 1>}
  n61 = Add(A=NamedRerouteDeclaration(None=@n17.G) {name=Msk_Curvature, nodeColor=<rgba 0 1 0.7422 1>}) {constB=-0.5}
  NamedRerouteDeclaration(None=Saturate(None=Multiply(A=@n61) {constB=-1})) {name=Curve_Pits, nodeColor=<rgba 1 0 0.8906 1>}
  NamedRerouteDeclaration(None=Saturate(None=@n61)) {name=Curve_Peaks, nodeColor=<rgba 1 0 0.8906 1>}
  n62 = TextureObjectParameter() {
   group=TopSide_V02
   parameterName=TopS_00B_Sharerd_Tex_Msk
   samplerType=SAMPLERTYPE_Masks
   sortPriority=249
   texture=&T_BlackMask
  }
  NamedRerouteDeclaration(None=@n62) {name=Tex_00B_Shared_Tex, nodeColor=<rgba 1 0.8906 0 1>}
  n63 = NamedRerouteUsage()
  n64 = NamedRerouteUsage()
  n65 = ScalarParameter() {defaultValue=0.1, group=Effect_EdgeWear_Dirt_V2_Pits, parameterName="Eff_Edge\\Dirt_Pits  Specular"}
  n66 = ScalarParameter() {group=Effect_EdgeWear_Dirt_V2_Pits, parameterName="Eff_Edge\\Dirt_Pits  Metalness"}
  n67 = LinearInterpolate(
   A=NamedRerouteUsage()
   B=NamedRerouteUsage()
   Alpha=ScalarParameter() {group=Effect_EdgeWear_Dirt_V2_Pits, parameterName="Eff_Edge\\Dirt_Pits  1=Root Normal"}
  )
  n68 = SetMaterialAttributes!(
   MaterialAttributes=NamedRerouteUsage()
   BaseColor=ComponentMask(None=@n64) {b=true, g=true, r=true}
   Roughness=ComponentMask(None=@n64) {a=true}
   Specular=@n65
   Metallic=@n66
   Normal=@n67
  ) {attributeSetTypes [
    #3
    #4
    #6
    #5
    #2
   ]}
  n69 = VectorParameter() {group=Effect_EdgeWear_Dirt_V2_Pits, parameterName="Eff_Edge\\Dirt_Pits  BaseColor", defaultValue=<rgba 0.0312 0.0168 0.0125 1>}
  n70 = VectorParameter() {
   group=Effect_EdgeWear_Dirt_V2_Pits
   parameterName="Eff_Edge\\Dirt_Pits EmissiveColor"
   channelNames {a=Brightness, b="", g="", r=""}
   defaultValue=<rgba 0 0 0 1>
  }
  n71 = SetMaterialAttributes!(
   MaterialAttributes=NamedRerouteUsage()
   BaseColor=@n69.RGB
   Roughness=ScalarParameter() {defaultValue=0.6, group=Effect_EdgeWear_Dirt_V2_Pits, parameterName="Eff_Edge\\Dirt_Pits  Roughness"}
   Specular=@n65
   Metallic=@n66
   Normal=@n67
   EmissiveColor=Multiply(A=@n70, B=@n70)
  ) {attributeSetTypes [
    #3
    #4
    #6
    #5
    #2
    #1
   ]}
  n72 = BlendMaterialAttributes(
   A=@n63
   B=StaticSwitchParameter(True=@n68, False=@n71) {group=Effect_EdgeWear_Dirt_V2_Pits, parameterName="Eff_Edge\\Dirt_Pits  Material Simple ---->Textures", sortPriority=0}
   Alpha=NamedRerouteUsage()
  ) {vertexAttributeBlendType=UseA}
  n73 = StaticSwitchParameter(True=@n72, False=NamedRerouteUsage()) {group=Effect_EdgeWear_Dirt_V2_Pits, parameterName="Eff_Edge\\Dirt_Pits   Pits\\Gmap\\Z Layer", sortPriority=0}
  n74 = NamedRerouteUsage()
  n75 = ComponentMask(None=@n74) {b=true, g=true, r=true}
  n76 = Constant()
  n77 = ScalarParameter() {group=Effect_EdgeWear_Dirt_V2_Edge, parameterName="Eff_Edge\\Dirt_Edge BaseColor--->Emissive"}
  n78 = ScalarParameter() {defaultValue=1, group=Effect_EdgeWear_Dirt_V2_Edge, parameterName="Eff_Edge\\Dirt_Edge  Specular"}
  n79 = ScalarParameter() {defaultValue=1, group=Effect_EdgeWear_Dirt_V2_Edge, parameterName="Eff_Edge\\Dirt_Edge  Metalness"}
  n80 = LinearInterpolate(
   A=NamedRerouteUsage()
   B=NamedRerouteUsage()
   Alpha=ScalarParameter() {group=Effect_EdgeWear_Dirt_V2_Edge, parameterName="Eff_Edge\\Dirt_Edge  1=Root Normal"}
  )
  n81 = SetMaterialAttributes(
   MaterialAttributes=NamedRerouteUsage()
   BaseColor=LinearInterpolate(A=@n75, B=@n76, Alpha=@n77)
   Roughness=ComponentMask(None=@n74) {a=true}
   Specular=@n78
   Metallic=@n79
   Normal=@n80
   EmissiveColor=LinearInterpolate(A=@n76, B=@n75, Alpha=@n77)
  ) {attributeSetTypes [
    #3
    #4
    #6
    #5
    #2
    #1
   ]}
  n82 = VectorParameter() {group=Effect_EdgeWear_Dirt_V2_Edge, parameterName="Eff_Edge\\Dirt_Edge  BaseColor", defaultValue=<rgba 0.6886 0.7713 0.8698 1>}
  n83 = VectorParameter() {
   group=Effect_EdgeWear_Dirt_V2_Edge
   parameterName="Eff_Edge\\Dirt_Edge  EmissiveColor"
   channelNames {a=Brightness, b="", g="", r=""}
   defaultValue=<rgba 0 0 0 1>
  }
  n84 = Multiply(A=@n83, B=@n83)
  n85 = NamedRerouteUsage()
  n86 = SetMaterialAttributes(
   MaterialAttributes=NamedRerouteUsage()
   BaseColor=@n82.RGB
   Roughness=ScalarParameter() {defaultValue=0.15, group=Effect_EdgeWear_Dirt_V2_Edge, parameterName="Eff_Edge\\Dirt_Edge  Roughness"}
   Specular=@n78
   Metallic=@n79
   Normal=@n80
   EmissiveColor=StaticSwitchParameter(True=Multiply(A=@n84, B=@n85), False=@n84) {group=Effect_EdgeWear_Dirt_V2_Edge_Reactive, parameterName="Eff_Edge\\Dirt_Edge_Emissive-Reactive", sortPriority=500}
  ) {attributeSetTypes [
    #3
    #4
    #6
    #5
    #2
    #1
   ]}
  n87 = StaticSwitchParameter(True=@n81, False=@n86) {group=Effect_EdgeWear_Dirt_V2_Edge, parameterName="Eff_Edge\\Dirt_Edge  Material Simple ---->Textures", sortPriority=0}
  n88 = NamedRerouteUsage()
  NamedRerouteDeclaration(None=StaticSwitchParameter(
    True=BlendMaterialAttributes(A=@n73, B=@n87, Alpha=@n88) {vertexAttributeBlendType=UseA}
    False=StaticSwitchParameter(True=@n72, False=@n63) {group=Effect_EdgeWear_Dirt_V2_Pits, parameterName="Eff_Edge\\Dirt_Pits   Pits\\Gmap\\Z Layer", sortPriority=0}
   ) {defaultValue=true, group=Effect_EdgeWear_Dirt_V2_Edge, parameterName="Eff_Edge\\Dirt_Pits   Edge", sortPriority=-1}) {name=MAT, nodeColor=<rgba 1 0.4922 0 1>}
  n89 = TextureObjectParameter() {group=Effect_EdgeWear_Dirt_V2_Edge, parameterName="Eff_Edge\\Dirt _Edge_BaseColor", texture=&T_BlacksRGB}
  NamedRerouteDeclaration(None=@n89) {name=Tex_Edge_BC, nodeColor=<rgba 1 0 0.9375 1>}
  n90 = ScalarParameter() {defaultValue=0.2, group=Effect_EdgeWear_Dirt_V2_Edge, parameterName="Eff_Edge\\Dirt_Edge_Hardness_Min"}
  n91 = ScalarParameter() {defaultValue=0.3, group=Effect_EdgeWear_Dirt_V2_Edge, parameterName="Eff_Edge\\Dirt_Edge_Hardness_Max"}
  n92 = NamedRerouteUsage()
  n93 = ScalarParameter() {defaultValue=0.5, group=Effect_EdgeWear_Dirt_V2_Edge, parameterName="Eff_Edge\\Dirt_Edge_Curvature_Position"}
  n94 = Power(
   Base=Multiply(A=Subtract(A=@n92, B=@n93), B=NamedRerouteUsage())
   Exp=ScalarParameter() {defaultValue=1, group=Effect_EdgeWear_Dirt_V2_Edge, parameterName="Eff_Edge\\Dirt_Edge_Curvature_Fallof"}
  )
  n95 = Max(A=@n94, B=NamedRerouteUsage())
  n96 = Constant() {r=1}
  n97 = VectorParameter() {group=Effect_EdgeWear_Dirt_V2_UV_Edge, parameterName="Eff_Edge\\Dirt_Edge_Grunge_Rotation", defaultValue=<rgba 0.3125 0.1356 0.1683 1>}
  n98 = ChannelMaskParameter(None=MaterialFunctionCall!(
    Texture=NamedRerouteUsage()
    Caculated_Bounds=NamedRerouteUsage()
    Pre_Skinned_Normals=NamedRerouteUsage()
    Blend_Contrast=@n96
    IsStatsB=NamedRerouteUsage()
    IsState=NamedRerouteUsage()
    Fill_alternate_Axis_Opacity=@n96
    "Fill_alternate_Axis_Color_A=Roughness"=Constant4Vector().A
    Offset
    Scale=NamedRerouteUsage()
    Offset_Pivot=NamedRerouteUsage()
    Rotation_Axis=@n97.RGB
   ) {materialFunction=&/MF_TopsSideProjection_Base_V2}.Combined) {group=Effect_EdgeWear_Dirt_V2_Edge, parameterName="Eff_Edge\\Dirt_Edge Grunge_Channel"}
  n99 = ScalarParameter() {defaultValue=0.8, group=Effect_EdgeWear_Dirt_V2_Edge, parameterName="Eff_Edge\\Dirt_Edge_Grunge_Position"}
  n100 = Add(
   A=StaticSwitch(
    True=StaticSwitchParameter(True=@n95, False=@n94) {group=Effect_EdgeWear_Dirt_V2_Edge, parameterName="Eff_Edge\\Dirt_Edge_TopSide_Msk_Add"}
    False=@n94
    Value=StaticBoolParameter() {group=TopSide_V02, parameterName=TopSideProjection_V02}
   )
   B=Multiply(
    A=Subtract(A=@n98.R, B=@n99) {constB=0}
    B=ScalarParameter() {defaultValue=1, group=Effect_EdgeWear_Dirt_V2_Edge, parameterName="Eff_Edge\\Dirt_Edge_Grunge_Strength"}
   )
  )
  n101 = ScalarParameter() {group=Effect_EdgeWear_Dirt_V2_Edge_Reactive, parameterName="Eff_Edge\\Dirt Audio_Opacity_Min"}
  n102 = ScalarParameter() {group=Effect_EdgeWear_Dirt_V2_Edge_Reactive, parameterName="Eff_Edge\\Dirt Audio_Opacity_Max"}
  n103 = NamedRerouteUsage()
  n104 = ScalarParameter() {defaultValue=-9999999, group=Effect_EdgeWear_Dirt_V2_Edge_Reactive, parameterName="Eff_Edge\\Dirt_Edge_Opacity_Time_BP", sortPriority=212}
  n105 = ScalarParameter() {group=Effect_EdgeWear_Dirt_V2_Edge_Reactive, parameterName="Eff_Edge\\Dirt_Edge_Opacity_Delay", sortPriority=211}
  n106 = ScalarParameter() {defaultValue=1, group=Effect_EdgeWear_Dirt_V2_Edge_Reactive, parameterName="Eff_Edge\\Dirt_Edge_Opacity_Duration", sortPriority=211}
  n107 = Divide(A=Subtract(A=Time(), B=Add(A=@n104, B=@n105)), B=@n106)
  n108 = ScalarParameter() {defaultValue=1, group=Effect_EdgeWear_Dirt_V2_Edge_Reactive, parameterName="Eff_Edge\\Dirt_Edge_Opacity_Falloff", sortPriority=211}
  n109 = LinearInterpolate(
   A=ScalarParameter() {defaultValue=1, group=Effect_EdgeWear_Dirt_V2_Edge_Reactive, parameterName="Eff_Edge\\Dirt_Edge_Opacity_Start"}
   B=ScalarParameter() {group=Effect_EdgeWear_Dirt_V2_Edge_Reactive, parameterName="Eff_Edge\\Dirt_Edge_Opacity_End"}
   Alpha=Power(Base=Saturate(None=@n107), Exp=@n108)
  )
  n110 = StaticSwitchParameter(True=Max(A=LinearInterpolate(A=@n101, B=@n102, Alpha=@n103), B=@n109), False=@n109) {group=Effect_EdgeWear_Dirt_V2_Edge_Reactive, parameterName="Eff_Edge\\Dirt Audio_Opacity-Reactive"}
  n111 = ScalarParameter() {defaultValue=1, group=Effect_EdgeWear_Dirt_V2_Edge, parameterName="Eff_Edge\\Dirt_Edge_Master_Opacity", sortPriority=1}
  NamedRerouteDeclaration(None=Multiply(
    A=SmoothStep(Min=@n90, Max=@n91, Value=@n100)
    B=StaticSwitchParameter(True=@n110, False=@n111) {group=Effect_EdgeWear_Dirt_V2_Edge_Reactive, parameterName="Eff_Edge\\Dirt_Edge_Opacity-Reactive"}
   )) {name=Edge_Msk, nodeColor=<rgba 0.1484 1 0 1>}
  NamedRerouteDeclaration(None=@n28.MaterialAttributes) {name=Root_Material, nodeColor=<rgba 0.7812 1 0 1>}
  NamedRerouteDeclaration(None=MaterialFunctionCall(
    Texture=NamedRerouteUsage()
    Caculated_Bounds=NamedRerouteUsage()
    Pre_Skinned_Normals=NamedRerouteUsage()
    Blend_Contrast=Constant() {r=1}
    IsStatsB=NamedRerouteUsage()
    IsState=NamedRerouteUsage()
    Fill_alternate_Axis_Opacity=Constant()
    "Fill_alternate_Axis_Color_A=Roughness"=Constant4Vector().A
    Scale=NamedRerouteUsage()
    Offset_Pivot=NamedRerouteUsage()
    Rotation_Axis=Constant()
   ) {materialFunction=&MF_TopsSideProjection_Base_V2}.Combined) {name=Edge_BaseColor, nodeColor=<rgba 1 0 0.4687 1>}
  n112 = ScalarParameter() {defaultValue=1, group=Effect_EdgeWear_Dirt_V2_UV_Edge, parameterName="Eff_Edge\\Dirt_Edge_Scale_Vehicle"}
  n113 = ScalarParameter() {defaultValue=1, group=Effect_EdgeWear_Dirt_V2_UV_Edge, parameterName="Eff_Edge\\Dirt_Edge_Vehicle_Side"}
  n114 = MaterialFunctionCall(Float4=NamedRerouteUsage()) {materialFunction=&BreakOutFloat4Components}
  n115 = LinearInterpolate(
   A=LinearInterpolate(
    A=ScalarParameter() {defaultValue=1, group=Effect_EdgeWear_Dirt_V2_UV_Edge, parameterName="Eff_Edge\\Dirt_Edge_Scale_Weapon"}
    B=LinearInterpolate(A=@n112, B=@n113, Alpha=@n114.G)
    Alpha=@n114.R
   )
   B=LinearInterpolate(
    A=ScalarParameter() {defaultValue=1, group=Effect_EdgeWear_Dirt_V2_UV_Edge, parameterName="Eff_Edge\\Dirt_Edge_Scale_Mech"}
    B=ScalarParameter() {defaultValue=1, group=Effect_EdgeWear_Dirt_V2_UV_Edge, parameterName="Eff_Edge\\Dirt_Edge_Scale_Character"}
    Alpha=@n114.A
   )
   Alpha=@n114.B
  )
  n116 = ScalarParameter() {defaultValue=1, group=Effect_EdgeWear_Dirt_V2_UV_Edge, parameterName="Eff_Edge\\Dirt_Edge_Scale_Backpack"}
  n117 = MaterialFunctionCall(Float2=NamedRerouteUsage()) {materialFunction=&BreakOutFloat2Components}
  NamedRerouteDeclaration(None=LinearInterpolate(
    A=LinearInterpolate(A=@n115, B=@n116, Alpha=@n117.R)
    B=ScalarParameter() {defaultValue=1, group=Effect_EdgeWear_Dirt_V2_UV_Edge, parameterName="Eff_Edge\\Dirt_Edge_Scale_Pickaxe"}
    Alpha=@n117.G
   )) {name=UVs_Scale_Edge, nodeColor=<rgba 0 0.7188 1 1>}
  n118 = ScalarParameter() {defaultValue=10, group=Effect_EdgeWear_Dirt_V2_Edge, parameterName="Eff_Edge\\Dirt_Edge_Curvature_Strength_Vehicle"}
  n119 = ScalarParameter() {defaultValue=2, group=Effect_EdgeWear_Dirt_V2_Edge, parameterName="Eff_Edge\\Dirt_Edge_Curvature_Strength_Vehicle_Side"}
  n120 = MaterialFunctionCall(Float4=NamedRerouteUsage()) {materialFunction=&BreakOutFloat4Components}
  n121 = LinearInterpolate(
   A=LinearInterpolate(
    A=ScalarParameter() {defaultValue=4, group=Effect_EdgeWear_Dirt_V2_Edge, parameterName="Eff_Edge\\Dirt_Edge_Curvature_Strength_Weapon"}
    B=LinearInterpolate(A=@n118, B=@n119, Alpha=@n120.G)
    Alpha=@n120.R
   )
   B=LinearInterpolate(
    A=ScalarParameter() {defaultValue=10, group=Effect_EdgeWear_Dirt_V2_Edge, parameterName="Eff_Edge\\Dirt_Edge_Curvature_Strength_Mech"}
    B=ScalarParameter() {defaultValue=6, group=Effect_EdgeWear_Dirt_V2_Edge, parameterName="Eff_Edge\\Dirt_Edge_Curvature_Strength_Character"}
    Alpha=@n120.A
   )
   Alpha=@n120.B
  )
  n122 = ScalarParameter() {defaultValue=10, group=Effect_EdgeWear_Dirt_V2_Edge, parameterName="Eff_Edge\\Dirt_Edge_Curvature_Strength_Backpack"}
  n123 = MaterialFunctionCall(Float2=NamedRerouteUsage()) {materialFunction=&BreakOutFloat2Components}
  NamedRerouteDeclaration(None=LinearInterpolate(
    A=LinearInterpolate(A=@n121, B=@n122, Alpha=@n123.R)
    B=ScalarParameter() {defaultValue=6, group=Effect_EdgeWear_Dirt_V2_Edge, parameterName="Eff_Edge\\Dirt_Edge_Curvature_Strength_Pickaxe"}
    Alpha=@n123.G
   )) {name=Curvature_Strength_Edge, nodeColor=<rgba 0 0.7188 1 1>}
  NamedRerouteDeclaration(None=@n17.A) {name=Msk_Gmap, nodeColor=<rgba 0 1 0.7422 1>}
  n124 = ScalarParameter() {defaultValue=0.1, group=Effect_EdgeWear_Dirt_V2_Pits, parameterName="Eff_Edge\\Dirt_Pits_Hardness_Min"}
  n125 = ScalarParameter() {defaultValue=1, group=Effect_EdgeWear_Dirt_V2_Pits, parameterName="Eff_Edge\\Dirt_Pits_Hardness_Max"}
  n126 = OneMinus(None=NamedRerouteUsage())
  n127 = ScalarParameter() {defaultValue=0.45, group=Effect_EdgeWear_Dirt_V2_Pits, parameterName="Eff_Edge\\Dirt_Pits_Curvature_Position"}
  n128 = Power(
   Base=Multiply(A=Subtract(A=@n126, B=@n127), B=NamedRerouteUsage())
   Exp=ScalarParameter() {defaultValue=1.25, group=Effect_EdgeWear_Dirt_V2_Pits, parameterName="Eff_Edge\\Dirt_Pits_Curvature_Fallof"}
  )
  n129 = StaticSwitchParameter(True=Max(A=@n128, B=Max(A=@n1, B=@n4)), False=@n128) {group=Effect_EdgeWear_Dirt_V2_Pits, parameterName="Eff_Edge\\Dirt_Pits_Add_Gmap"}
  n130 = Max(A=NamedRerouteUsage(), B=@n129)
  n131 = Constant() {r=1}
  n132 = VectorParameter() {group=Effect_EdgeWear_Dirt_V2_UV_Pits, parameterName="Eff_Edge\\Dirt_Pit_Grunge_Rotation", defaultValue=<rgba 0.1345 0.3672 0.6123 1>}
  n133 = ChannelMaskParameter(None=MaterialFunctionCall(
    Texture=NamedRerouteUsage()
    Caculated_Bounds=NamedRerouteUsage()
    Pre_Skinned_Normals=NamedRerouteUsage()
    Blend_Contrast=@n131
    IsStatsB=NamedRerouteUsage()
    IsState=NamedRerouteUsage()
    Fill_alternate_Axis_Opacity=@n131
    "Fill_alternate_Axis_Color_A=Roughness"=Constant4Vector().A
    Scale=NamedRerouteUsage()
    Offset_Pivot=NamedRerouteUsage()
    Rotation_Axis=@n132.RGB
   ) {materialFunction=&MF_TopsSideProjection_Base_V2}.Combined) {group=Effect_EdgeWear_Dirt_V2_Pits, maskChannel=Green, parameterName="Eff_Edge\\Dirt_Pits Grunge_Channel", defaultValue=<rgba 0 1 0 0>}
  n134 = ScalarParameter() {defaultValue=0.2, group=Effect_EdgeWear_Dirt_V2_Pits, parameterName="Eff_Edge\\Dirt_Pits_Grunge_Position"}
  n135 = Add(
   A=StaticSwitch(
    True=StaticSwitchParameter(True=@n130, False=@n129) {group=Effect_EdgeWear_Dirt_V2_Pits, parameterName="Eff_Edge\\Dirt_Pits_TopSide_Msk_Add", sortPriority=50}
    False=@n129
    Value=StaticBoolParameter() {group=TopSide_V02, parameterName=TopSideProjection_V02}
   )
   B=Multiply(
    A=Subtract(A=@n133.R, B=@n134) {constB=0}
    B=ScalarParameter() {defaultValue=0.5, group=Effect_EdgeWear_Dirt_V2_Pits, parameterName="Eff_Edge\\Dirt_Pits_Grunge_Strength"}
   )
  )
  NamedRerouteDeclaration(None=Multiply(
    A=SmoothStep(Min=@n124, Max=@n125, Value=@n135)
    B=ScalarParameter() {defaultValue=1, group=Effect_EdgeWear_Dirt_V2_Pits, parameterName="Eff_Edge\\Dirt_Pits_Master_Opacity", sortPriority=1}
   )) {name=Pits_Msk, nodeColor=<rgba 0.1484 1 0 1>}
  n136 = ScalarParameter() {defaultValue=2, group=Effect_EdgeWear_Dirt_V2_UV_Pits, parameterName="Eff_Edge\\Dirt_Pits_Grunge_Scale_Vehicle"}
  n137 = ScalarParameter() {defaultValue=4, group=Effect_EdgeWear_Dirt_V2_UV_Pits, parameterName="Eff_Edge\\Dirt_Pits_Grunge_Vehicle_Side"}
  n138 = MaterialFunctionCall(Float4=NamedRerouteUsage()) {materialFunction=&BreakOutFloat4Components}
  n139 = LinearInterpolate(
   A=LinearInterpolate(
    A=ScalarParameter() {defaultValue=3, group=Effect_EdgeWear_Dirt_V2_UV_Pits, parameterName="Eff_Edge\\Dirt_Pits_Grunge_Scale_Weapon"}
    B=LinearInterpolate(A=@n136, B=@n137, Alpha=@n138.G)
    Alpha=@n138.R
   )
   B=LinearInterpolate(
    A=ScalarParameter() {defaultValue=4, group=Effect_EdgeWear_Dirt_V2_UV_Pits, parameterName="Eff_Edge\\Dirt_Pits_Grunge_Scale_Mech"}
    B=ScalarParameter() {defaultValue=2, group=Effect_EdgeWear_Dirt_V2_UV_Pits, parameterName="Eff_Edge\\Dirt_Pits_Grunge_Scale_Character"}
    Alpha=@n138.A
   )
   Alpha=@n138.B
  )
  n140 = ScalarParameter() {defaultValue=1.5, group=Effect_EdgeWear_Dirt_V2_UV_Pits, parameterName="Eff_Edge\\Dirt_Pits_Grunge_Scale_Backpack"}
  n141 = MaterialFunctionCall(Float2=NamedRerouteUsage()) {materialFunction=&BreakOutFloat2Components}
  NamedRerouteDeclaration(None=LinearInterpolate(
    A=LinearInterpolate(A=@n139, B=@n140, Alpha=@n141.R)
    B=ScalarParameter() {defaultValue=2, group=Effect_EdgeWear_Dirt_V2_UV_Pits, parameterName="Eff_Edge\\Dirt_Pits_Grunge_Scale_Pickaxe"}
    Alpha=@n141.G
   )) {name=UVs_Scale_Pits_Grunge, nodeColor=<rgba 0 0.7188 1 1>}
  n142 = TextureObjectParameter() {
   group=TopSide_V02
   parameterName=TopS_00A_Sharerd_Tex_Msk
   samplerType=SAMPLERTYPE_Masks
   sortPriority=249
   texture=&T_BlackMask
  }
  NamedRerouteDeclaration(None=@n142) {name=Tex_00A_Shared_Tex, nodeColor=<rgba 1 0.8906 0 1>}
  NamedRerouteDeclaration(None=@n16.RGBA) {name=CM_Mask, nodeColor=<rgba 1 0 0.3047 1>}
  n143 = VectorParameter() {group=Effect_EdgeWear_Dirt_V2_UV_Pits, parameterName="Eff_Edge\\Dirt_Pits_Offset", defaultValue=<rgba 0 0 0 1>}
  NamedRerouteDeclaration(None=@n143.RGB) {name=UVs_Offset_Pits, nodeColor=<rgba 0 0.7188 1 1>}
  NamedRerouteDeclaration(None=MaterialFunctionCall(
    Texture=NamedRerouteUsage()
    Caculated_Bounds=NamedRerouteUsage()
    Pre_Skinned_Normals=NamedRerouteUsage()
    Blend_Contrast=Constant() {r=1}
    IsStatsB=NamedRerouteUsage()
    IsState=NamedRerouteUsage()
    Fill_alternate_Axis_Opacity=Constant()
    "Fill_alternate_Axis_Color_A=Roughness"=Constant4Vector().A
    Scale=NamedRerouteUsage()
    Offset_Pivot=NamedRerouteUsage()
    Rotation_Axis=Constant()
   ) {materialFunction=&MF_TopsSideProjection_Base_V2}.Combined) {name=Pits_BaseColor, nodeColor=<rgba 1 0 0.4688 1>}
  NamedRerouteDeclaration(None=@n98.R) {name=edge_Grundge, nodeColor=<rgba 0.7813 0 1 1>}
  n144 = TextureObjectParameter() {group=Effect_EdgeWear_Dirt_V2_Pits, parameterName="Eff_Edge\\Dirt _Pits_BaseColor", texture=&T_BlacksRGB}
  NamedRerouteDeclaration(None=@n144) {name=Tex_Pits_BC, nodeColor=<rgba 1 0 0.9375 1>}
  n145 = ScalarParameter() {defaultValue=1, group=Effect_EdgeWear_Dirt_V2_UV_Pits, parameterName="Eff_Edge\\Dirt_Pits_Scale_Vehicle"}
  n146 = ScalarParameter() {defaultValue=1, group=Effect_EdgeWear_Dirt_V2_UV_Pits, parameterName="Eff_Edge\\Dirt_Pits_Vehicle_Side"}
  n147 = MaterialFunctionCall(Float4=NamedRerouteUsage()) {materialFunction=&BreakOutFloat4Components}
  n148 = LinearInterpolate(
   A=LinearInterpolate(
    A=ScalarParameter() {defaultValue=1, group=Effect_EdgeWear_Dirt_V2_UV_Pits, parameterName="Eff_Edge\\Dirt_Pits_Scale_Weapon"}
    B=LinearInterpolate(A=@n145, B=@n146, Alpha=@n147.G)
    Alpha=@n147.R
   )
   B=LinearInterpolate(
    A=ScalarParameter() {defaultValue=1, group=Effect_EdgeWear_Dirt_V2_UV_Pits, parameterName="Eff_Edge\\Dirt_Pits_Scale_Mech"}
    B=ScalarParameter() {defaultValue=1, group=Effect_EdgeWear_Dirt_V2_UV_Pits, parameterName="Eff_Edge\\Dirt_Pits_Scale_Character"}
    Alpha=@n147.A
   )
   Alpha=@n147.B
  )
  n149 = ScalarParameter() {defaultValue=1, group=Effect_EdgeWear_Dirt_V2_UV_Pits, parameterName="Eff_Edge\\Dirt_Pits_Scale_Backpack"}
  n150 = MaterialFunctionCall(Float2=NamedRerouteUsage()) {materialFunction=&BreakOutFloat2Components}
  NamedRerouteDeclaration(None=LinearInterpolate(
    A=LinearInterpolate(A=@n148, B=@n149, Alpha=@n150.R)
    B=ScalarParameter() {defaultValue=1, group=Effect_EdgeWear_Dirt_V2_UV_Pits, parameterName="Eff_Edge\\Dirt_Pits_Scale_Pickaxe"}
    Alpha=@n150.G
   )) {name=UVs_Scale_Pits, nodeColor=<rgba 0 0.7188 1 1>}
  n151 = ScalarParameter() {defaultValue=6, group=Effect_EdgeWear_Dirt_V2_Pits, parameterName="Eff_Edge\\Dirt_Pits_Curvature_Strength_Vehicle"}
  n152 = ScalarParameter() {defaultValue=5, group=Effect_EdgeWear_Dirt_V2_Pits, parameterName="Eff_Edge\\Dirt_Pits_Curvature_Strength_Vehicle_Side"}
  n153 = MaterialFunctionCall(Float4=NamedRerouteUsage()) {materialFunction=&BreakOutFloat4Components}
  n154 = LinearInterpolate(
   A=LinearInterpolate(
    A=ScalarParameter() {defaultValue=5, group=Effect_EdgeWear_Dirt_V2_Pits, parameterName="Eff_Edge\\Dirt_Pits_Curvature_Strength_Weapon"}
    B=LinearInterpolate(A=@n151, B=@n152, Alpha=@n153.G)
    Alpha=@n153.R
   )
   B=LinearInterpolate(
    A=ScalarParameter() {defaultValue=7, group=Effect_EdgeWear_Dirt_V2_Pits, parameterName="Eff_Edge\\Dirt_Pits_Curvature_Strength_Mech"}
    B=ScalarParameter() {defaultValue=4, group=Effect_EdgeWear_Dirt_V2_Pits, parameterName="Eff_Edge\\Dirt_Pits_Curvature_Strength_Character"}
    Alpha=@n153.A
   )
   Alpha=@n153.B
  )
  n155 = ScalarParameter() {defaultValue=4, group=Effect_EdgeWear_Dirt_V2_Pits, parameterName="Eff_Edge\\Dirt_Pits_Curvature_Strength_Backpack"}
  n156 = MaterialFunctionCall(Float2=NamedRerouteUsage()) {materialFunction=&BreakOutFloat2Components}
  NamedRerouteDeclaration(None=LinearInterpolate(
    A=LinearInterpolate(A=@n154, B=@n155, Alpha=@n156.R)
    B=ScalarParameter() {defaultValue=4, group=Effect_EdgeWear_Dirt_V2_Pits, parameterName="Eff_Edge\\Dirt_Pits_Curvature_Strength_Pickaxe"}
    Alpha=@n156.G
   )) {name=Curvature_Strength_Pits, nodeColor=<rgba 0 0.7188 1 1>}
  n157 = ScalarParameter() {defaultValue=0.2, group=Effect_EdgeWear_Dirt_V2_Pits, parameterName="Eff_Edge\\Dirt_Pits_Gmap_Strength_Vehicle"}
  n158 = ScalarParameter() {defaultValue=1, group=Effect_EdgeWear_Dirt_V2_Pits, parameterName="Eff_Edge\\Dirt_Pits_Gmap_Strength_Vehicle_Side"}
  n159 = MaterialFunctionCall(Float4=NamedRerouteUsage()) {materialFunction=&BreakOutFloat4Components}
  n160 = LinearInterpolate(
   A=LinearInterpolate(
    A=ScalarParameter() {defaultValue=1, group=Effect_EdgeWear_Dirt_V2_Pits, parameterName="Eff_Edge\\Dirt_Pits_Gmap_Strength_Weapon"}
    B=LinearInterpolate(A=@n157, B=@n158, Alpha=@n159.G)
    Alpha=@n159.R
   )
   B=LinearInterpolate(
    A=ScalarParameter() {defaultValue=1, group=Effect_EdgeWear_Dirt_V2_Pits, parameterName="Eff_Edge\\Dirt_Pits_Gmap_Strength_Mech"}
    B=ScalarParameter() {defaultValue=1, group=Effect_EdgeWear_Dirt_V2_Pits, parameterName="Eff_Edge\\Dirt_Pits_Gmap_Strength_Character"}
    Alpha=@n159.A
   )
   Alpha=@n159.B
  )
  n161 = ScalarParameter() {defaultValue=0.05, group=Effect_EdgeWear_Dirt_V2_Pits, parameterName="Eff_Edge\\Dirt_Pits_Gmap_Strength_Backpack"}
  n162 = MaterialFunctionCall(Float2=NamedRerouteUsage()) {materialFunction=&BreakOutFloat2Components}
  NamedRerouteDeclaration(None=LinearInterpolate(
    A=LinearInterpolate(A=@n160, B=@n161, Alpha=@n162.R)
    B=ScalarParameter() {defaultValue=1, group=Effect_EdgeWear_Dirt_V2_Pits, parameterName="Eff_Edge\\Dirt_Pits_Gmap_Strength_Pickaxe"}
    Alpha=@n162.G
   )) {name=Gmap_Strength_Pits, nodeColor=<rgba 0 0.7188 1 1>}
  n163 = VectorParameter() {
   group=Effect_EdgeWear_Dirt_V2_Pits
   parameterName="Eff_Edge\\Dirt_Pits_Z_Weapon"
   channelNames {a="", b=Strength, g=Falloff, r=Offset}
   defaultValue=<rgba -0.42 0.15 0.75 0.18>
  }
  n164 = VectorParameter() {
   group=Effect_EdgeWear_Dirt_V2_Pits
   parameterName="Eff_Edge\\Dirt_Pits_Z_Vehicle"
   channelNames {a="", b=Strength, g=Falloff, r=Offset}
   defaultValue=<rgba -0.4 0.1 0.2 1>
  }
  n165 = AppendVector(A=@n164, B=@n164)
  n166 = VectorParameter() {
   group=Effect_EdgeWear_Dirt_V2_Pits
   parameterName="Eff_Edge\\Dirt_Pits_Z_VehicleSide"
   channelNames {a="", b=Strength, g=Falloff, r=Offset}
   defaultValue=<rgba -0.4 0.15 1 1>
  }
  n167 = AppendVector(A=@n166, B=@n166)
  n168 = MaterialFunctionCall(Float4=NamedRerouteUsage()) {materialFunction=&BreakOutFloat4Components}
  n169 = VectorParameter() {
   group=Effect_EdgeWear_Dirt_V2_Pits
   parameterName="Eff_Edge\\Dirt_Pits_Z_Mech"
   channelNames {a="", b=Strength, g=Falloff, r=Offset}
   defaultValue=<rgba -0.5 0.2 0.5 1>
  }
  n170 = VectorParameter() {
   group=Effect_EdgeWear_Dirt_V2_Pits
   parameterName="Eff_Edge\\Dirt_Pits_Z_Character"
   channelNames {a="", b=Strength, g=Falloff, r=Offset}
   defaultValue=<rgba -0.6 0.2 1 1>
  }
  n171 = LinearInterpolate(
   A=LinearInterpolate(A=AppendVector(A=@n163, B=@n163), B=LinearInterpolate(A=@n165, B=@n167, Alpha=@n168.G), Alpha=@n168.R)
   B=LinearInterpolate(A=AppendVector(A=@n169, B=@n169), B=AppendVector(A=@n170, B=@n170), Alpha=@n168.A)
   Alpha=@n168.B
  )
  n172 = VectorParameter() {
   group=Effect_EdgeWear_Dirt_V2_Pits
   parameterName="Eff_Edge\\Dirt_Pits_Z_Backpack"
   channelNames {a="", b=Strength, g=Falloff, r=Offset}
   defaultValue=<rgba -0.5 0.4 1 1>
  }
  n173 = AppendVector(A=@n172, B=@n172)
  n174 = MaterialFunctionCall(Float2=NamedRerouteUsage()) {materialFunction=&BreakOutFloat2Components}
  n175 = VectorParameter() {
   group=Effect_EdgeWear_Dirt_V2_Pits
   parameterName="Eff_Edge\\Dirt_Pits_Z_Pickaxe"
   channelNames {a="", b=Strength, g=Falloff, r=Offset}
   defaultValue=<rgba -0.33 0.3 1 1>
  }
  NamedRerouteDeclaration(None=LinearInterpolate(A=LinearInterpolate(A=@n171, B=@n173, Alpha=@n174.R), B=AppendVector(A=@n175, B=@n175), Alpha=@n174.G)) {name=Pits_Z_setings, nodeColor=<rgba 0 0.625 1 1>}
  n176 = ScalarParameter() {defaultValue=1, group=Effect_EdgeWear_Dirt_V2_Edge_Reactive, parameterName="Eff_Edge\\Dirt_Edge_Emissive_Start", sortPriority=501}
  n177 = Time()
  n178 = ScalarParameter() {defaultValue=-9999999, group=Effect_EdgeWear_Dirt_V2_Edge_Reactive, parameterName="Eff_Edge\\Dirt_Edge_Emissive_Time_BP", sortPriority=502}
  n179 = Add(
   A=@n178
   B=ScalarParameter() {group=Effect_EdgeWear_Dirt_V2_Edge_Reactive, parameterName="Eff_Edge\\Dirt_Edge_Emissive_Delay", sortPriority=502}
  )
  n180 = ScalarParameter() {defaultValue=1, group=Effect_EdgeWear_Dirt_V2_Edge_Reactive, parameterName="Eff_Edge\\Dirt_Edge_Emissive_Duration", sortPriority=502}
  n181 = Saturate(None=Divide(A=Subtract(A=@n177, B=@n179), B=@n180))
  n182 = ScalarParameter() {defaultValue=1, group=Effect_EdgeWear_Dirt_V2_Edge_Reactive, parameterName="Eff_Edge\\Dirt_Edge_Emissive_Falloff", sortPriority=502}
  NamedRerouteDeclaration(None=LinearInterpolate(
    A=@n176
    B=ScalarParameter() {group=Effect_EdgeWear_Dirt_V2_Edge_Reactive, parameterName="Eff_Edge\\Dirt_Edge_Emissive_End", sortPriority=501}
    Alpha=Power(Base=@n181, Exp=@n182)
   )) {name=Reactive_Emissive, nodeColor=<rgba 0 1 1 1>}
  NamedRerouteDeclaration(None=@n27.EmissiveColor) {name=In_Emissive, nodeColor=<rgba 0.4062 0 1 1>}
  NamedRerouteDeclaration(None=FunctionInput(Preview=Constant4Vector().RGBA) {bUsePreviewValueAsDefault=true, inputName=IsState, inputType=FunctionInput_Vector4, sortPriority=4}) {name=IsState, nodeColor=<rgba 0.1497 1 0.1633 1>}
  NamedRerouteDeclaration(None=FunctionInput(Preview=Constant2Vector()) {bUsePreviewValueAsDefault=true, inputName=IsState_B, inputType=FunctionInput_Vector2, sortPriority=4}) {name=IsState_B, nodeColor=<rgba 0.1497 1 0.1633 1>}
  n183 = PreSkinnedLocalBounds()
  n184 = Subtract(A=VertexInterpolator(VS=PreSkinnedPosition()).PS, B=@n183.Min)
  n185 = MaterialFunctionCall(Color=Subtract(A=@n183.Max, B=@n183.Max)) {materialFunction=&SplitComponents}
  n186 = Max(A=Max(A=@n185.R, B=@n185.R), B=@n185.B)
  NamedRerouteDeclaration(None=FunctionInput(Preview=Divide(A=@n184, B=@n186)) {bUsePreviewValueAsDefault=true, inputName=Calcated_Bounds, sortPriority=11}) {name=Caclcated_Bounds, nodeColor=<rgba 0 0.9297 1 1>}
  n187 = TextureObjectParameter() {
   group=Effect_EdgeWear_Dirt_V2
   parameterName="Eff_Edge\\Dirt  MSk_Tex"
   samplerType=SAMPLERTYPE_Masks
   texture=/Game/Global/Textures/Default/Noise/T_Noise_Cloud_Mask
  }
  NamedRerouteDeclaration(None=StaticSwitchParameter(
    True=@n142
    False=StaticSwitchParameter(True=@n62, False=@n187) {group=Effect_EdgeWear_Dirt_V2, parameterName="Eff_Edge\\Dirt  MSk_Tex------>TopS_00B_Sharerd_Tex_Msk"}
   ) {group=Effect_EdgeWear_Dirt_V2, parameterName="Eff_Edge\\Dirt  MSk_Tex------>TopS_00A_Sharerd_Tex_Msk"}) {name=Tex_MSk, nodeColor=<rgba 1 0.8906 0 1>}
  n188 = PreSkinnedNormal()
  NamedRerouteDeclaration(None=FunctionInput(Preview=VertexInterpolator(VS=@n188).PS) {bUsePreviewValueAsDefault=true, inputName=Pre-Skinned Local Normal, sortPriority=10}) {name=PreSkin_Normal, nodeColor=<rgba 0.8281 1 0 1>}
  NamedRerouteDeclaration(None=StaticBoolParameter() {group=Effect_EdgeWear_Dirt_V2_Edge_Reactive, parameterName="Eff_Edge\\Dirt Audio Reactive Opacity"}) {name=Audio_Opacity, nodeColor=<rgba 0 0.4141 1 1>}
  NamedRerouteDeclaration(None=@n27.Normal) {name=In_Normal, nodeColor=<rgba 0.4062 0 1 1>}
  NamedRerouteUsage()
  ScalarParameter() {defaultValue=1, group=Effect_EdgeWear_Dirt_V2_Edge, parameterName="Eff_Edge\\Dirt_Edge_Curvature_Strength"}
 ]}
