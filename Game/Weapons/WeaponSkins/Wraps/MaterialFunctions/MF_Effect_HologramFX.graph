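# Weapon wrap material function that, behind a static switch, adds an animated hologram effect to the incoming material's emissive.
# MF_Wrap_HologramFX generates scanlines, strokes and glitch, with flicker driven by a Blueprint-timestamped ramp.
# UVs come from virtual-plane or side projection, with tiling blended by IsState for weapon, vehicle, mech and character.
# Primary/accent, reactive and left/right decal, and TopSide projection masks limit where the effect shows.
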MaterialFunction /Game/Weapons/WeaponSkins/Wraps/MaterialFunctions/MF_Effect_HologramFX
graph {nodes [
  Constant4Vector()
  n1 = Time() {bOverride_Period=true, period=6000}
  n2 = ScalarParameter() {defaultValue=1, group=Effect_Hologram, parameterName=Holo_Speed}
  n3 = StaticSwitchParameter(True=NamedRerouteUsage(), False=NamedRerouteUsage()) {defaultValue=true, group=Effect_Hologram, parameterName="SideUVs--->VPUVs", sortPriority=0}
  n4 = ScalarParameter() {group=Effect_Hologram_Reactive, parameterName=Holo_Glitch_Start_Value}
  n5 = ScalarParameter() {group=Effect_Hologram_Reactive, parameterName=Holo_Glitch_End_Value}
  n6 = Subtract(
   A=Time()
   B=Add(
    A=ScalarParameter() {defaultValue=-9999999, group=Effect_Hologram_Reactive, parameterName=Holo_Glitch_TimeFrom-BP------, sortPriority=50}
    B=ScalarParameter() {group=Effect_Hologram_Reactive, parameterName=Holo_Glitch_Delay, sortPriority=50}
   )
  )
  n7 = ScalarParameter() {defaultValue=0.25, group=Effect_Hologram_Reactive, parameterName=Holo_Glitch_Duration, sortPriority=50}
  n8 = Power(
   Base=Saturate(None=Divide(A=@n6, B=@n7))
   Exp=ScalarParameter() {defaultValue=1, group=Effect_Hologram_Reactive, parameterName=Holo_Glitch_Fallof, sortPriority=50}
  )
  n9 = ScalarParameter() {defaultValue=2, group=Effect_Hologram, parameterName=Holo_Stroke_Size}
  n10 = NamedRerouteUsage()
  n11 = VectorParameter() {group=Effect_Hologram, parameterName=Holo_ColorTint, defaultValue=<rgba 1 1 1 0>}
  n12 = ScalarParameter() {defaultValue=1, group=Effect_Hologram, parameterName="Holo_Decal LR  1=on", sortPriority=0}
  n13 = FunctionInput(Preview=Constant() {r=1}) {inputName=Decal_LR, inputType=FunctionInput_Scalar}
  n14 = StaticSwitchParameter(True=Max(A=@n12, B=OneMinus(None=@n13)), False=Constant() {r=1}) {group=Customization, parameterName=Function_Left_Right_Decal_V2}
  n15 = VectorParameter() {
   group=Effect_Hologram
   parameterName=Holo_Decal_Multi_MSk
   sortPriority=0
   channelNames {a="DecalD     1-On", b="DecalC     1-On", g="DecalB     1-On", r="DecalA     1-On"}
   defaultValue=<rgba 1 1 1 1>
  }
  MaterialExpressionReroute_0 = Reroute(None=FunctionInput(Preview=Constant4Vector() {constant=<rgba 1 1 1 0>}.RGBA) {inputName=Decal_React, inputType=FunctionInput_Vector4})
  n16 = Convert(None=@MaterialExpressionReroute_0) {
   nodeName=Break Float4
   convertInputs [{
     type=Vector4
     defaultValue=<rgba 0 0 0 1>
     expressionInput {
      expression=&self:MaterialExpressionReroute_0
      inputName=None
      mask=0
      maskA=0
      maskB=0
      maskG=0
      maskR=0
      outputIndex=0
     }
    }]
   convertOutputs [{type=Scalar}, {type=Scalar}, {type=Scalar}, {type=Scalar}]
  }
  n17 = Max(A=Min(A=@n15."DecalA     1-On", B=@n16), B=Min(A=@n15."DecalB     1-On", B=@n16))
  n18 = Max(A=Min(A=@n15."DecalC     1-On", B=@n16), B=Min(A=@n15."DecalD     1-On", B=@n16))
  MaterialExpressionOneMinus_3 = OneMinus(None=@MaterialExpressionReroute_0)
  n19 = Convert(None=@MaterialExpressionOneMinus_3) {
   nodeName=Break Float4
   convertInputs [{
     type=Vector4
     defaultValue=<rgba 0 0 0 1>
     expressionInput {
      expression=&self:MaterialExpressionOneMinus_3
      inputName=None
      mask=0
      maskA=0
      maskB=0
      maskG=0
      maskR=0
      outputIndex=0
     }
    }]
   convertOutputs [{type=Scalar}, {type=Scalar}, {type=Scalar}, {type=Scalar}]
  }
  n20 = Min(
   A=Min(A=Max(A=@n15."DecalA     1-On", B=@n19), B=Max(A=@n15."DecalB     1-On", B=@n19))
   B=Min(A=Max(A=@n15."DecalC     1-On", B=@n19), B=Max(A=@n15."DecalD     1-On", B=@n19))
  )
  n21 = Constant() {r=1}
  n22 = ScalarParameter() {defaultValue=1, group=Effect_Hologram, parameterName="Holo_Primary_Msk 1=On 0=Off", sortPriority=0}
  n23 = TextureSample() {samplerType=SAMPLERTYPE_Masks, texture=/Game/Weapons/FORT_Rifles/Textures/T_Scar_CM}
  n24 = FunctionInput(Preview=@n23.R) {inputName=Mask Primary, inputType=FunctionInput_Scalar, sortPriority=24, previewValue=<1 0 0 1>}
  n25 = FunctionInput(Preview=@n23.B) {inputName=Mask Secondary, inputType=FunctionInput_Scalar, sortPriority=24, previewValue=<1 0 0 1>}
  n26 = ScalarParameter() {defaultValue=1, group=Effect_Hologram, parameterName="Holo_Accent_Msk 1=On 0=Off", sortPriority=0}
  n27 = Max(
   A=Max(A=Min(A=@n22, B=@n24), B=Min(A=@n25, B=@n26))
   B=ScalarParameter() {group=Effect_Hologram, parameterName="Holo_Disable_Masks 0=Off  1=ON", sortPriority=0}
  )
  n28 = Max(
   A=StaticSwitchParameter(True=Max(A=@n17, B=@n18), False=Constant()) {group=Customization, parameterName=Effect_Multi_Decal_Reactive}
   B=Min(A=StaticSwitchParameter(True=@n20, False=@n21) {group=Customization, parameterName=Effect_Multi_Decal_Reactive}, B=Saturate(None=@n27))
  )
  n29 = Min(A=@n12, B=@n13)
  n30 = Constant()
  n31 = LinearInterpolate(
   A=LinearInterpolate(B=NamedRerouteUsage(), Alpha=ScalarParameter() {group=Effect_Hologram, parameterName=Holo_RootEmissive, sortPriority=0})
   B=MaterialFunctionCall!(
    Time=Multiply(A=@n1, B=@n2) {constB=0}
    MeshScaleIn
    UVForScanline=@n3
    CustomFlicker=ScalarParameter() {defaultValue=1, group=Effect_Hologram_Reactive, parameterName=Holo_Glitch_Custome_Value}
    UseCustomFlicker=LinearInterpolate(A=@n4, B=@n5, Alpha=@n8)
    UVForStroke=Multiply(A=@n3, B=@n9)
    UVForGlitch=@n3
    BaseUVIn=@n3
    EmissiveIn=Multiply(A=@n10, B=@n11.RGB)
    OpacityIn
    WPOIn
   ) {materialFunction=&/MF_Wrap_HologramFX}.EmissiveOut
   Alpha=Max(A=Min(A=@n14, B=@n28), B=StaticSwitchParameter(True=@n29, False=@n30) {group=Customization, parameterName=Function_Left_Right_Decal_V2})
  )
  n32 = Multiply(
   A=@n31
   B=ChannelMaskParameter(None=FunctionInput() {bUsePreviewValueAsDefault=true, inputName=TS_Mask_01, inputType=FunctionInput_Vector4, sortPriority=100}) {group=Effect_Hologram, parameterName=Holo_TopSide_01_Mask_Channel, sortPriority=2}.R
  )
  n33 = FunctionInput() {bUsePreviewValueAsDefault=true, inputName=TS_Mask_02, inputType=FunctionInput_Vector4, sortPriority=100}
  n34 = Constant() {r=1}
  n35 = StaticSwitch(
   True=Add(
    A=ScalarParameter() {defaultValue=1, group=Effect_Hologram, parameterName="Holo_Use_TopS Emissive_01_Msk  0=On", sortPriority=6}
    B=FunctionInput() {bUsePreviewValueAsDefault=true, inputName=TopS Emissive_01, inputType=FunctionInput_Scalar, sortPriority=100}
   )
   False=@n34
   Value=StaticBoolParameter() {group=TopSide_V02_Layer Emissive_02, parameterName=TopS Emissive_02}
  )
  n36 = StaticSwitch(
   True=Add(
    A=FunctionInput() {bUsePreviewValueAsDefault=true, inputName=TopS Emissive_02, inputType=FunctionInput_Scalar, sortPriority=100}
    B=ScalarParameter() {defaultValue=1, group=Effect_Hologram, parameterName="Holo_Use_TopS Emissive_02_Msk  0=On", sortPriority=6}
   )
   False=@n34
   Value=StaticBoolParameter() {group=TopSide_V02_Layer Emissive_01, parameterName=TopS Emissive_01}
  )
  n37 = Multiply(A=@n31, B=Saturate(None=Add(A=@n35, B=@n36)))
  n38 = StaticSwitchParameter(
   True=Multiply(A=@n31, B=ChannelMaskParameter(None=@n33) {group=Effect_Hologram, parameterName=Holo_TopSide_02_Mask_Channel, sortPriority=2}.R)
   False=@n37
  ) {group=Effect_Hologram, parameterName=Holo_TopSide_02_Mask, sortPriority=1}
  n39 = SetMaterialAttributes!(
   MaterialAttributes=NamedRerouteUsage()
   EmissiveColor=StaticSwitchParameter(True=StaticSwitchParameter(True=@n32, False=@n38) {group=Effect_Hologram, parameterName=Holo_TopSide_01_Mask, sortPriority=1}, False=@n37) {group=Customization, parameterName=TopSideProjection_V02}
   BaseColor
   Roughness
   Specular
   Metallic
   Normal
  ) {attributeSetTypes [
    #1
    #2
    #3
    #4
    #5
    #6
   ]}
  n40 = Constant3Vector() {constant=<rgba 0.15 0.15 0.15 0>}
  n41 = SetMaterialAttributes(EmissiveColor=@n40, BaseColor=@n40) {attributeSetTypes [
    #1
    #2
    #3
    #4
    #5
    #6
   ]}
  n42 = FunctionInput(Preview=@n41) {inputName=Material, inputType=FunctionInput_MaterialAttributes}
  FunctionOutput(None=StaticSwitchParameter(True=@n39, False=@n42) {group=Customization, parameterName=Effect_HologramFX})
  n43 = FunctionInput(Preview=Constant4Vector() {constant=<rgba 0 0 0 -1>}.RGBA) {bUsePreviewValueAsDefault=true, inputName=Object-SafeZone-Virtual-Pivot, inputType=FunctionInput_Vector4, sortPriority=105}
  n44 = FunctionInput(Preview=Constant()) {bUsePreviewValueAsDefault=true, inputName=Object-SafeZone-VP_Charcter_Preview_Switch, inputType=FunctionInput_Scalar, sortPriority=105}
  n45 = MaterialFunctionCall!(
   Camera _pos
   Axis
   Scale=Multiply(
    A=MaterialFunctionCall() {materialFunction=/Engine/Functions/Engine_MaterialFunctions02/WorldPositionOffset/ObjectScale}.Scale Z
    B=NamedRerouteUsage()
   )
   Scale_Multplyer
   Object-SafeZone-Virtual-Pivot=@n43
   Object-SafeZone-Virtual-CH-Switch=@n44
   SinglePlane Normal Axis
   SinglePlane X Axis
   PlaneCenter_distortion
  ) {materialFunction=&/MF_SafeZone_VirtualPlane}
  Multiply(A=@n45.Uvs, B=NamedRerouteUsage())
  n46 = GetMaterialAttributes(None=@n42) {attributeGetTypes [#2, #1]}
  NamedRerouteDeclaration(None=@n46.BaseColor) {name=BaseColor, nodeColor=<rgba 0.9922 0 1 1>}
  NamedRerouteDeclaration(None=@n46.MaterialAttributes) {name=Root_Material, nodeColor=<rgba 0.7812 1 0 1>}
  n47 = ScalarParameter() {group=Effect_Hologram, parameterName=Holo_Rotation_Weapon}
  n48 = MaterialFunctionCall(Float4=NamedRerouteUsage()) {materialFunction=&BreakOutFloat4Components}
  n49 = LinearInterpolate(
   A=ScalarParameter() {group=Effect_Hologram, parameterName=Holo_Rotation_Vehicle}
   B=ScalarParameter() {group=Effect_Hologram, parameterName=Holo_Rotation_VehicleSide}
   Alpha=@n48.G
  )
  n50 = ScalarParameter() {defaultValue=0.25, group=Effect_Hologram, parameterName=Holo_Rotation_Mech}
  n51 = ScalarParameter() {group=Effect_Hologram, parameterName=Holo_Rotation_Character}
  NamedRerouteDeclaration(None=LinearInterpolate(A=LinearInterpolate(A=@n47, B=@n49, Alpha=@n48.R), B=LinearInterpolate(A=@n50, B=@n51, Alpha=@n48.A), Alpha=@n48.B)) {name=Side_Rotation, nodeColor=<rgba 1 0.7266 0 1>}
  n52 = ScalarParameter() {defaultValue=1, group=Effect_Hologram, parameterName=Holo_Side_TileRate_Weapon}
  n53 = MaterialFunctionCall(Float4=NamedRerouteUsage()) {materialFunction=&BreakOutFloat4Components}
  n54 = LinearInterpolate(
   A=ScalarParameter() {defaultValue=1, group=Effect_Hologram, parameterName=Holo_Side_TileRate_Vehicle}
   B=ScalarParameter() {defaultValue=1, group=Effect_Hologram, parameterName=Holo_Side_TileRate_VehicleSide}
   Alpha=@n53.G
  )
  n55 = ScalarParameter() {defaultValue=1, group=Effect_Hologram, parameterName=Holo_Side_TileRate_Mech}
  n56 = ScalarParameter() {defaultValue=1, group=Effect_Hologram, parameterName=Holo_Side_TileRate_Character}
  NamedRerouteDeclaration(None=LinearInterpolate(A=LinearInterpolate(A=@n52, B=@n54, Alpha=@n53.R), B=LinearInterpolate(A=@n55, B=@n56, Alpha=@n53.A), Alpha=@n53.B)) {name=Side_TileRate, nodeColor=<rgba 1 0.7266 0 1>}
  n57 = ScalarParameter() {defaultValue=-256, group=Effect_Hologram, parameterName=Holo_TileRate_Weapon}
  n58 = MaterialFunctionCall(Float4=NamedRerouteUsage()) {materialFunction=&BreakOutFloat4Components}
  n59 = LinearInterpolate(
   A=ScalarParameter() {defaultValue=-512, group=Effect_Hologram, parameterName=Holo_TileRate_Vehicle}
   B=ScalarParameter() {defaultValue=-512, group=Effect_Hologram, parameterName=Holo_TileRate_VehicleSide}
   Alpha=@n58.G
  )
  n60 = ScalarParameter() {defaultValue=-512, group=Effect_Hologram, parameterName=Holo_TileRate__Mech}
  n61 = ScalarParameter() {defaultValue=-256, group=Effect_Hologram, parameterName=Holo_TileRate_Character}
  NamedRerouteDeclaration(None=LinearInterpolate(A=LinearInterpolate(A=@n57, B=@n59, Alpha=@n58.R), B=LinearInterpolate(A=@n60, B=@n61, Alpha=@n58.A), Alpha=@n58.B)) {name=VP_Tilerate, nodeColor=<rgba 0 1 0.1797 1>}
  n62 = StaticBool() {value=true}
  n63 = WorldPosition() {worldPositionShaderOffset=WPT_ExcludeAllShaderOffsets}
  NamedRerouteDeclaration(None=Divide(
    A=MaterialFunctionCall!(UseLocalSpace=@n62, WorldPosition=@n63.XYZ, NormalOverride) {materialFunction=/Game/Environments/Apollo/Sets/Skyfire/Materials/MaterialFunctions/MF_WorldAlignedUVs_HardEdge}.Result
    B=ScalarParameter() {defaultValue=128, group=Effect_Hologram, parameterName=Holo_TileRate_WS}
   )) {name=CheapTri_UVS, nodeColor=<rgba 0 0.9062 1 1>}
  NamedRerouteDeclaration(None=@n45.Uvs) {name=VP_UVs, nodeColor=<rgba 0 0.4844 1 1>}
  NamedRerouteDeclaration(None=FunctionInput(Preview=Constant4Vector().RGBA) {bUsePreviewValueAsDefault=true, inputName=IsState, inputType=FunctionInput_Vector4, sortPriority=4}) {name=IsState, nodeColor=<rgba 0.1497 1 0.1633 1>}
  NamedRerouteDeclaration(None=FunctionInput(Preview=Constant2Vector()) {bUsePreviewValueAsDefault=true, inputName=IsState_B, inputType=FunctionInput_Vector2, sortPriority=4}) {name=IsState_B, nodeColor=<rgba 0.1497 1 0.1633 1>}
  n64 = PreSkinnedLocalBounds()
  n65 = Subtract(A=VertexInterpolator(VS=PreSkinnedPosition()).PS, B=@n64.Min)
  n66 = MaterialFunctionCall(Color=Subtract(A=@n64.Max, B=@n64.Max)) {materialFunction=&SplitComponents}
  n67 = Max(A=Max(A=@n66.R, B=@n66.R), B=@n66.B)
  NamedRerouteDeclaration(None=FunctionInput(Preview=Divide(A=@n65, B=@n67)) {bUsePreviewValueAsDefault=true, inputName=Calcated_Bounds, sortPriority=11}) {name=Caclcated_Bounds, nodeColor=<rgba 0 0.9297 1 1>}
  NamedRerouteDeclaration(None=@n46.EmissiveColor) {name=EmissiveColor, nodeColor=<rgba 1 0.5625 0 1>}
  n68 = PreSkinnedNormal()
  NamedRerouteDeclaration(None=FunctionInput(Preview=VertexInterpolator(VS=@n68).PS) {bUsePreviewValueAsDefault=true, inputName=Pre-Skinned Local Normal, sortPriority=10}) {name=PreSkin_Normal, nodeColor=<rgba 0.8281 1 0 1>}
  n69 = MaterialFunctionCall(Float4=NamedRerouteUsage()) {materialFunction=&BreakOutFloat4Components}
  n70 = MaterialFunctionCall!(
   For_Safe_Zones
   Normal_Map
   Caculated_Bounds=NamedRerouteUsage()
   Pre_Skinned_Normals=NamedRerouteUsage()
   Is Vehicle=@n69.R
   Is VehicleSide=@n69.R
   Is Mech=@n69.R
   Vehicle_Rotation
   UV_Rotation Bypass=NamedRerouteUsage()
   Mech_Rotation
   Uvs_Offset
   Offset-after rotation
   Uvs_TileRate=NamedRerouteUsage()
  ) {materialFunction=&/MF_Projection_Side_Normal}
  NamedRerouteDeclaration(None=StaticSwitchParameter(True=@n45.Uvs, False=NamedRerouteDeclaration(None=@n70.UVs) {name=Side_UVs, nodeColor=<rgba 0.7344 1 0 1>}) {defaultValue=true, group=Effect_Hologram, parameterName="SideUVs--->VPUVs", sortPriority=0}) {name=uvs, nodeColor=<rgba 0.3828 1 0 1>}
  ScalarParameter() {defaultValue=256, group=Effect_Hologram, parameterName=Holo_VP_Scale}
 ]}
