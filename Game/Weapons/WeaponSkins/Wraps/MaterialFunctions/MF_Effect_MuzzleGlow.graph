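# Weapon wrap material function that adds an optional muzzle glow to incoming material attributes, gated by the Effect_MuzzleGlow static switch.
# Pre-skinned position and local bounds build a barrel gradient and falloff, combined with primary, secondary and crevice mask inputs.
# Glow color comes from a scrolling curve-atlas lookup or three-color lerp, with edge tinting.
# It drives emissive, base-color darkening and optional roughness/specular/metallic, with optional time-since-event reactive fading.
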MaterialFunction /Game/Weapons/WeaponSkins/Wraps/MaterialFunctions/MF_Effect_MuzzleGlow
graph {nodes [
  n1 = TextureSample() {samplerType=SAMPLERTYPE_Masks, texture=/Game/Weapons/FORT_Rifles/Textures/T_Scar_CM}
  n2 = FunctionInput(Preview=@n1.R) {bUsePreviewValueAsDefault=true, inputName=Mask Primary, inputType=FunctionInput_Scalar, sortPriority=24, previewValue=<1 0 0 1>}
  n3 = FunctionInput(Preview=@n1.B) {bUsePreviewValueAsDefault=true, inputName=Mask Secondary, inputType=FunctionInput_Scalar, sortPriority=24, previewValue=<1 0 0 1>}
  n4 = Add(A=@n2, B=@n3)
  BlendMaterialAttributes(
   A=BlendMaterialAttributes(A=NamedRerouteUsage(), B=NamedRerouteUsage(), Alpha=Saturate(None=@n4)) {vertexAttributeBlendType=UseA}
   B=NamedRerouteUsage()
   Alpha=NamedRerouteDeclaration(None=@n3) {name=Msk_Accent, nodeColor=<rgba 0 1 0.7656 1>}
  ) {vertexAttributeBlendType=UseA}
  n5 = Add(A=ComponentMask(None=NamedRerouteUsage()) {g=true}) {constB=-1}
  Ceil(None=@n5)
  n6 = NamedRerouteUsage()
  n7 = NamedRerouteUsage()
  n8 = MaterialFunctionCall(Float4=NamedRerouteUsage()) {materialFunction=&BreakOutFloat4Components}
  n9 = MaterialFunctionCall(Float2=NamedRerouteUsage()) {materialFunction=&BreakOutFloat2Components}
  n10 = Max(A=Max(A=Max(A=@n8.R, B=@n8.R), B=Max(A=@n8.B, B=@n8.B)), B=Max(A=@n9.R, B=@n9.R))
  n11 = TextureSample() {samplerType=SAMPLERTYPE_Normal, texture=/Game/Weapons/FORT_Rifles/T_SCAR_N}
  n12 = Constant3Vector() {constant=<rgba 1 0 0 0>}
  n13 = GetMaterialAttributes(None=FunctionInput(Preview=SetMaterialAttributes!(MaterialAttributes, Normal=@n11.RGB, BaseColor=@n12) {attributeSetTypes [#1, #2]}) {inputName=Material, inputType=FunctionInput_MaterialAttributes}) {attributeGetTypes [
    #2
    #3
    #1
    #4
    #5
   ]}
  FunctionOutput(None=StaticSwitchParameter(
    True=BlendMaterialAttributes(A=@n6, B=@n7, Alpha=@n10) {vertexAttributeBlendType=UseA}
    False=NamedRerouteDeclaration(None=@n13.MaterialAttributes) {name=Root_Material, nodeColor=<rgba 0.7812 1 0 1>}
   ) {group=Customization, parameterName=Effect_MuzzleGlow})
  n14 = NamedRerouteUsage()
  n15 = ScalarParameter() {group=Effect_Muzzle_Glow, parameterName=MzGlow_Color_Gradent_Curvature_Amount}
  n16 = PreSkinnedLocalBounds()
  n17 = PreSkinnedLocalBounds()
  n18 = Subtract(A=Add(A=@n17.Max, B=NamedRerouteUsage()), B=Add(A=@n17.Min, B=NamedRerouteUsage()))
  n19 = Subtract(A=NamedRerouteUsage(), B=Add(A=@n16.Min, B=@n18))
  n20 = Subtract(A=Subtract(A=@n16.Max, B=@n16.Max), B=@n18)
  n21 = ComponentMask(None=NamedRerouteDeclaration(None=Divide(A=@n19, B=@n20)) {name=Bounds, nodeColor=<rgba 1 0 0.7031 1>}) {g=true}
  n22 = LinearInterpolate(A=@n21, B=OneMinus(None=@n21), Alpha=ScalarParameter() {group=Effect_Muzzle_Glow, parameterName="MzGlow__Invert  1=on"})
  n23 = Saturate(None=@n22)
  n24 = ScalarParameter() {defaultValue=1, group=Effect_Muzzle_Glow, parameterName=MzGlow_Color_Gradent_Pow}
  n25 = Add(
   A=Add(A=Multiply(A=@n14, B=@n15), B=Power(Base=@n23, Exp=@n24))
   B=ScalarParameter() {group=Effect_Muzzle_Glow, parameterName=MzGlow__Muzzle_CC_Offset}
  )
  n26 = ScalarParameter() {defaultValue=1, group=Effect_Muzzle_Glow, parameterName=MzGlow__Muzzle_CC_TileRate}
  n27 = ScalarParameter() {group=Effect_Muzzle_Glow, parameterName=MzGlow__Muzzle_CC_Speed}
  n28 = Time()
  n29 = Frac(None=Add(A=Multiply(A=@n25, B=@n26), B=Multiply(A=@n27, B=@n28)))
  n30 = CurveAtlasRowParameter(CurveTime=@n29) {
   atlas=/Game/Weapons/WeaponSkins/Wraps/Textures/CruveAtlas_Wraps_01
   curve=/Game/Weapons/WeaponSkins/Wraps/Textures/HunterD/CC_Rainbow
   defaultValue=42
   group=Effect_Muzzle_Glow
   parameterName=MzGlow__Muzzle_CC
  }
  n31 = MaterialFunctionCall!(
   Index=ScalarParameter() {defaultValue=42, group=Effect_Muzzle_Glow, parameterName=MzGlow__Muzzle_CC Index}
   CCinput=@n30
   CCinput_Alpha=@n30
   UV=@n29
  ) {materialFunction=&/MF_Global_CC_Tex}
  n32 = Multiply(A=@n31.3V, B=@n31.3V)
  n33 = StaticSwitchParameter(
   True=@n32
   False=MaterialFunctionCall(
    A=VectorParameter() {group=Effect_Muzzle_Glow, parameterName=MzGlow_Color_A, defaultValue=<rgba 1 0 0 1>}.RGB
    B=VectorParameter() {group=Effect_Muzzle_Glow, parameterName=MzGlow_Color_B, defaultValue=<rgba 0.0107 1 0 1>}.RGB
    C=VectorParameter() {group=Effect_Muzzle_Glow, parameterName=MzGlow_Color_C, defaultValue=<rgba 0.0488 0 1 1>}.RGB
    Alpha=@n25
   ) {materialFunction=&Lerp_3Color}.Result
  ) {group=Effect_Muzzle_Glow, parameterName=MzGlow__ColorCurve}
  n34 = ScalarParameter() {defaultValue=1, group=Effect_Muzzle_Glow, parameterName=MzGlow_Sine_Pow}
  n35 = Add(
   A=Multiply(A=Power(Base=@n22, Exp=@n34), B=ScalarParameter() {defaultValue=5, group=Effect_Muzzle_Glow, parameterName=MzGlow_Sine_Tilerate})
   B=Multiply(A=ScalarParameter() {group=Effect_Muzzle_Glow, parameterName=MzGlow_Sine_Speed}, B=Time())
  )
  n36 = Power(
   Base=ConstantBiasScale(None=Sine(None=@n35))
   Exp=ScalarParameter() {defaultValue=1, group=Effect_Muzzle_Glow, parameterName=MzGlow_Sine_Falloff}
  )
  LinearInterpolate(
   A=@n33
   B=Multiply(A=@n33, B=VectorParameter() {group=Effect_Muzzle_Glow, parameterName=MzGlow_Sine_Multiplyer_Color, defaultValue=<rgba 0 0 0 1>}.RGB)
   Alpha=Multiply(A=Frac(None=@n36), B=ScalarParameter() {group=Effect_Muzzle_Glow, parameterName=MzGlow_Sine__Strength})
  )
  n37 = MaterialFunctionCall(Color=@n20) {materialFunction=&SplitComponents}
  Max(A=Max(A=@n37.R, B=@n37.R), B=@n37.B)
  n38 = Saturate(None=Add(A=NamedRerouteUsage(), B=ScalarParameter() {defaultValue=1, group=Effect_Muzzle_Glow, parameterName="MzGlow__Accent_Mask   0=On"}))
  Multiply(A=OneMinus(None=NamedRerouteUsage()), B=@n38)
  n39 = FunctionInput(Preview=@n1.G) {bUsePreviewValueAsDefault=true, inputName=Mask Crevise, inputType=FunctionInput_Scalar, sortPriority=24, previewValue=<1 0 0 1>}
  NamedRerouteDeclaration(None=@n39) {name=Msk_Curvature, nodeColor=<rgba 0 1 0.7422 1>}
  n40 = ScalarParameter() {defaultValue=0.001, group=Effect_Muzzle_Glow, parameterName=MzGlow__Barrel_Falloff}
  NamedRerouteDeclaration(None=Saturate(None=Divide(A=@n5, B=@n40) {constB=0})) {name=Msk_Bounds, nodeColor=<rgba 1 0.0469 0 1>}
  NamedRerouteDeclaration(None=@n13.BaseColor) {name=Root_BC, nodeColor=<rgba 0.4062 0 1 1>}
  NamedRerouteDeclaration(None=@n2) {name=Msk_Prim, nodeColor=<rgba 0 1 0.7656 1>}
  n41 = ScalarParameter() {defaultValue=1, group=Effect_Muzzle_Glow_reactive, parameterName=MzGlow_Brightness_Start}
  n42 = ScalarParameter() {group=Effect_Muzzle_Glow_reactive, parameterName=MzGlow_Brightness_End}
  n43 = Subtract(
   A=Time()
   B=Add(
    A=ScalarParameter() {defaultValue=-9999999, group=Wrap_Reactive, parameterName=TimeFrom-BP----------------, sortPriority=600}
    B=ScalarParameter() {defaultValue=0.25, group=Effect_Muzzle_Glow_reactive, parameterName=MzGlow_Reactive_Delay}
   )
  )
  n44 = ScalarParameter() {defaultValue=0.25, group=Effect_Muzzle_Glow_reactive, parameterName=MzGlow_Reactive_Duration}
  n45 = Power(
   Base=Saturate(None=Divide(A=@n43, B=@n44))
   Exp=ScalarParameter() {defaultValue=1, group=Effect_Muzzle_Glow_reactive, parameterName=MzGlow_Reactive_Falloff}
  )
  n46 = StaticBoolParameter() {group=Effect_Muzzle_Glow_reactive, parameterName=MzGlow__Reactive}
  NamedRerouteDeclaration(None=StaticSwitch(True=LinearInterpolate(A=@n41, B=@n42, Alpha=@n45), False=Constant() {r=1}, Value=@n46)) {name=Reactive, nodeColor=<rgba 0.5 0 1 1>}
  n47 = ScalarParameter() {group=Effect_Muzzle_Glow_reactive, parameterName=MzGlow_Darkens_Start}
  n48 = ScalarParameter() {defaultValue=1, group=Effect_Muzzle_Glow_reactive, parameterName=MzGlow_Darkens_End}
  NamedRerouteDeclaration(None=StaticSwitch(True=LinearInterpolate(A=@n47, B=@n48, Alpha=@n45), False=Constant() {r=1}, Value=@n46)) {name=Reactive_Darken, nodeColor=<rgba 0.5 0 1 1>}
  NamedRerouteDeclaration(None=@n46) {name=Reactive State, nodeColor=<rgba 1 0 0.0469 1>}
  n49 = GetMaterialAttributes(None=NamedRerouteUsage()) {attributeGetTypes [
    #2
    #3
    #1
    #4
    #5
    #6
   ]}
  n50 = Multiply(A=@n38, B=NamedRerouteUsage())
  n51 = VectorParameter() {group=Effect_Muzzle_Glow, parameterName="MzGlow_Edge_Color     A=1=UseColor", defaultValue=<rgba 0.004 0 0.4323 0>}
  n52 = NamedRerouteUsage()
  n53 = Saturate(None=Multiply(A=Add(A=@n52) {constB=-0.5}) {constB=4})
  n54 = LinearInterpolate(A=@n33, B=@n51.RGB, Alpha=Saturate(None=Multiply(A=@n51.A, B=@n53)))
  n55 = LinearInterpolate(A=@n54, B=Multiply(A=ScalarParameter() {defaultValue=1, group=Effect_Muzzle_Glow, parameterName=MzGlow_Edge_Brightness}, B=@n54), Alpha=@n53)
  n56 = NamedRerouteUsage()
  n57 = Add(A=Multiply(A=@n49.EmissiveColor, B=OneMinus(None=@n50)) {constA=1}, B=Multiply(A=@n50, B=Multiply(A=@n55, B=@n56)))
  n58 = StaticSwitch(
   True=NamedRerouteUsage()
   False=ScalarParameter() {defaultValue=1, group=Effect_Muzzle_Glow, parameterName="MzGlow___Darken_BaseC   0=on"}
   Value=NamedRerouteUsage()
  )
  n59 = OneMinus(None=@n50)
  n60 = Multiply(A=NamedRerouteUsage(), B=Saturate(None=Add(A=@n58, B=@n59)))
  n61 = SetMaterialAttributes!(
   MaterialAttributes=NamedRerouteUsage()
   BaseColor=@n60
   EmissiveColor=@n57
   Roughness=LinearInterpolate(
    A=@n49.Roughness
    B=ScalarParameter() {defaultValue=0.3, group=Effect_Muzzle_Glow_Custum_Roughness, parameterName=MzGlow_Roughness}
    Alpha=@n50
   )
   Specular=LinearInterpolate(A=@n49.Specular, B=ScalarParameter() {defaultValue=0.5, group=Effect_Muzzle_Glow_Custum_Roughness, parameterName=MzGlow_Spec}, Alpha=@n50)
   Metallic=LinearInterpolate(A=@n49.Metallic, B=ScalarParameter() {group=Effect_Muzzle_Glow_Custum_Roughness, parameterName=MzGlow_Metalic}, Alpha=@n50)
  ) {attributeSetTypes [
    #2
    #5
    #3
    #6
    #4
   ]}
  n62 = NamedRerouteUsage()
  NamedRerouteDeclaration(None=StaticSwitchParameter(
    True=@n61
    False=SetMaterialAttributes!(MaterialAttributes=@n62, BaseColor=@n60, EmissiveColor=@n57) {attributeSetTypes [#2, #5]}
   ) {group=Effect_Muzzle_Glow_Custum_Roughness, parameterName="MzGlow_Custom_Rough,Spec,Metalic"}) {name=Mat_Base, nodeColor=<rgba 0 0.3906 1 1>}
  NamedRerouteDeclaration(None=@n32) {name=Muzzle_Color, nodeColor=<rgba 0 0.7422 1 1>}
  n63 = VertexInterpolator(VS=PreSkinnedPosition())
  NamedRerouteDeclaration(None=FunctionInput(Preview=@n63.PS) {bUsePreviewValueAsDefault=true, inputName=Pre-Skinned Position, sortPriority=10}) {name=PreSkin_Position, nodeColor=<rgba 0.8281 1 0 1>}
  n64 = Constant3Vector() {constant=<rgba 0 -23.0279 0 0>}
  NamedRerouteDeclaration(None=FunctionInput(Preview=@n64) {bUsePreviewValueAsDefault=true, inputName=Bounds_Max, sortPriority=101}) {name=Bounds_Max, nodeColor=<rgba 0 0.9297 1 1>}
  n65 = Constant3Vector() {constant=<rgba 0 2.7572 0 0>}
  NamedRerouteDeclaration(None=FunctionInput(Preview=@n65) {bUsePreviewValueAsDefault=true, inputName=Bounds_min, sortPriority=102}) {name=Bounds_Min, nodeColor=<rgba 0 0.9297 1 1>}
  NamedRerouteDeclaration(None=FunctionInput(Preview=Constant4Vector().RGBA) {bUsePreviewValueAsDefault=true, inputName=IsState, inputType=FunctionInput_Vector4, sortPriority=4}) {name=IsState, nodeColor=<rgba 0.1497 1 0.1633 1>}
  NamedRerouteDeclaration(None=FunctionInput(Preview=Constant2Vector()) {bUsePreviewValueAsDefault=true, inputName=IsState_B, inputType=FunctionInput_Vector2, sortPriority=4}) {name=IsState_B, nodeColor=<rgba 0.1497 1 0.1633 1>}
  n66 = PreSkinnedLocalBounds()
  n67 = Add(A=@n66.Min, B=@n65)
  n68 = Subtract(A=@n63.PS, B=@n67)
  n69 = MaterialFunctionCall(Color=Subtract(A=Add(A=@n66.Max, B=@n64), B=@n67)) {materialFunction=&SplitComponents}
  n70 = Max(A=Max(A=@n69.R, B=@n69.R), B=@n69.B)
  NamedRerouteDeclaration(None=FunctionInput(Preview=Divide(A=@n68, B=@n70)) {bUsePreviewValueAsDefault=true, inputName=Calcated_Bounds, sortPriority=11}) {name=Caclcated_Bounds, nodeColor=<rgba 0 0.9297 1 1>}
  NamedRerouteDeclaration(None=@n13.EmissiveColor) {name=Root_Emissive, nodeColor=<rgba 0.4062 0 1 1>}
  n71 = PreSkinnedNormal()
  NamedRerouteDeclaration(None=FunctionInput(Preview=VertexInterpolator(VS=@n71).PS) {bUsePreviewValueAsDefault=true, inputName=Pre-Skinned Local Normal, sortPriority=10}) {name=PreSkin_Normal, nodeColor=<rgba 0.8281 1 0 1>}
  NamedRerouteDeclaration(None=@n13.Normal) {name=Root_Normal, nodeColor=<rgba 0.4062 0 1 1>}
  NamedRerouteUsage()
  ScalarParameter() {defaultValue=-9999999, group=TopSideProjection_Emissive_Reactive, parameterName=Time_BP}
  ScalarParameter() {defaultValue=2, group=Effect_Muzzle_Glow, parameterName=MzGlow_Sine_Multiplyer}
  SetMaterialAttributes!(MaterialAttributes, BaseColor, Roughness, Specular, Metallic, Normal, EmissiveColor) {attributeSetTypes [
    #2
    #3
    #6
    #4
    #1
    #5
   ]}
  StaticSwitchParameter() {group=Effect_Muzzle_Glow_reactive, parameterName=MzGlow__Reactive}
 ]}
