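# Material function for weapon wraps that adds a super-levelling effect, revealing a simple override material as the level rises, behind a static switch.
# Blueprint-supplied time animates the level step, which is normalized, optionally quantized and remapped by a curve atlas into a mask radius.
# Side-projected coordinates, tuned for weapon, vehicle or mech, form that mask, broken up by world-aligned noise.
# The mask blends the materials, with an optional eased emissive edge glow.
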MaterialFunction /Game/Weapons/WeaponSkins/Wraps/MaterialFunctions/MF_Effect_SuperLevellingMask
graph {nodes [
  n1 = GetMaterialAttributes(None=FunctionInput(Preview=MakeMaterialAttributes!(
     BaseColor
     Metallic
     Specular
     Roughness
     Anisotropy
     EmissiveColor
     Opacity
     OpacityMask
     Normal
     Tangent
     WorldPositionOffset
     SubsurfaceColor
     ClearCoat
     ClearCoatRoughness
     AmbientOcclusion
     Refraction
     CustomizedUVs_0
     CustomizedUVs_1
     CustomizedUVs_2
     CustomizedUVs_3
     CustomizedUVs_4
     CustomizedUVs_5
     CustomizedUVs_6
     CustomizedUVs_7
     PixelDepthOffset
     ShadingModel
     Displacement
    )) {bUsePreviewValueAsDefault=true, inputName=Material, inputType=FunctionInput_MaterialAttributes}) {attributeGetTypes [#1]}
  n2 = PreSkinnedPosition()
  n3 = PreSkinnedLocalBounds()
  n4 = Divide(A=Subtract(A=VertexInterpolator(VS=@n2).PS, B=@n3.Min), B=Subtract(A=@n3.Max, B=@n3.Max))
  n5 = FunctionInput() {inputName=Is_States, inputType=FunctionInput_Vector4}
  n6 = MaterialFunctionCall(Float4=@n5) {materialFunction=&BreakOutFloat4Components}
  n7 = LinearInterpolate(A=@n4, B=FunctionInput(Preview=@n4) {inputName="Calculated Bounds "}, Alpha=@n6.A)
  n8 = FunctionInput(Preview=VertexInterpolator(VS=PreSkinnedNormal()).PS) {bUsePreviewValueAsDefault=true, inputName=Local Normal}
  n9 = ScalarParameter() {group=Effect_SuperLevelling, parameterName=SL_UV_Rotation_Vehicle, sortPriority=200}
  n10 = ScalarParameter() {defaultValue=0.25, group=Effect_SuperLevelling, parameterName=SL_UV_Rotation_Mech, sortPriority=200}
  n11 = VectorParameter() {group=Effect_SuperLevelling, parameterName=SL_Offset_Weapon, sortPriority=200, defaultValue=<rgba 0.5 0 0 0>}
  n12 = VectorParameter() {group=Effect_SuperLevelling, parameterName=SL_Offset_Vehicle, sortPriority=200, defaultValue=<rgba 0.5 0 0 1>}
  n13 = VectorParameter() {group=Effect_SuperLevelling, parameterName=SL_Offset_VehicleSide, sortPriority=200, defaultValue=<rgba 0.5 0 0 1>}
  n14 = VectorParameter() {group=Effect_SuperLevelling, parameterName=SL_Offset_Mech, sortPriority=200, defaultValue=<rgba 0.5 0 0 1>}
  n15 = LinearInterpolate(A=LinearInterpolate(A=@n11.RGB, B=LinearInterpolate(A=@n12.RGB, B=@n13.RGB, Alpha=@n6.G), Alpha=@n6.R), B=@n14.RGB, Alpha=@n6.B)
  n16 = ScalarParameter() {group=Effect_SuperLevelling, parameterName=SL_CurrentLevel, sortPriority=3}
  n17 = Time()
  n18 = ScalarParameter() {defaultValue=-99999, group=Effect_SuperLevelling, parameterName=SL_GameTimeFromBP, sortPriority=100}
  n19 = ScalarParameter() {defaultValue=2, group=Effect_SuperLevelling, parameterName=SL_Animation_Duration, sortPriority=101}
  n20 = Saturate(None=Divide(A=Subtract(A=@n17, B=@n18), B=Add(A=@n19) {constB=0}))
  n21 = LinearInterpolate(A=Subtract(A=@n16), B=@n16, Alpha=@n20)
  n22 = ScalarParameter() {group=Effect_SuperLevelling, parameterName=SL_StartAtLevel, sortPriority=3}
  n23 = ScalarParameter() {defaultValue=6, group=Effect_SuperLevelling, parameterName=SL_StopAtLevel, sortPriority=3}
  n24 = Divide(A=Subtract(A=Clamp(None=@n21, Min=@n22, Max=@n23), B=@n22), B=Subtract(A=@n23, B=@n22))
  n25 = Max(A=Ceil(None=ScalarParameter() {defaultValue=1, group=Effect_SuperLevelling, parameterName=SL_QuantizationSteps, sliderMax=100, sliderMin=1}))
  n26 = Multiply(A=@n24, B=@n25)
  n27 = Subtract(A=@n25)
  n28 = LinearInterpolate(A=@n24, B=Divide(A=Floor(None=@n26), B=@n25), Alpha=Saturate(None=Ceil(None=@n27)))
  n29 = CurveAtlasRowParameter(CurveTime=@n28) {
   atlas=/Game/Weapons/WeaponSkins/Wraps/Textures/CruveAtlas_Wraps_01
   curve=/Game/Weapons/WeaponSkins/Wraps/Textures/Curve_SuperlevelMask_Kills
   defaultValue=58
   group=Effect_SuperLevelling
   parameterName=SL_SuperLevelMaskCurve
  }
  n30 = ChannelMaskParameter(None=MaterialFunctionCall!(
    Index=ScalarParameter() {defaultValue=58, group=Effect_SuperLevelling, parameterName="SL_SuperLevelMaskCurve  Index"}
    CCinput=@n29
    CCinput_Alpha=@n29
    UV=@n28
   ) {materialFunction=&/MF_Global_CC_Tex}.4v) {group=Effect_SuperLevelling, parameterName=SL_SuperLevelCurveChannel}
  n31 = SphereMask(
   A=ComponentMask(None=MaterialFunctionCall!(
     For_Safe_Zones
     Normal_Map
     Caculated_Bounds=@n7
     Pre_Skinned_Normals=@n8
     Is Vehicle=@n6.R
     Is VehicleSide=@n6.R
     Is Mech=@n6.R
     Vehicle_Rotation=@n9
     UV_Rotation Bypass
     Mech_Rotation=@n10
     Uvs_Offset=@n15
     Offset-after rotation
     Uvs_TileRate
    ) {materialFunction=&/MF_Projection_Side_Normal}.UVs) {r=true}
   B=Constant() {r=0.5}
   Radius=LinearInterpolate(
    A=ScalarParameter() {group=Effect_SuperLevelling, parameterName=SL_StartRadius}
    B=ScalarParameter() {defaultValue=1.5, group=Effect_SuperLevelling, parameterName=SL_EndRadius}
    Alpha=@n30.R
   )
   Hardness=ScalarParameter() {defaultValue=0.5, group=Effect_SuperLevelling, parameterName=SL_CoverMaskHardness}
  )
  n32 = TextureObjectParameter() {
   group=Effect_SuperLevelling
   parameterName=SL_TransitionNoise
   texture=/Game/Characters/Player/Male/Medium/Bodies/M_MED_Viper/Textures/Glitch/T_SoftNoise
  }
  n33 = ScalarParameter() {defaultValue=50, group=Effect_SuperLevelling, parameterName=SL_Noise_TileRate_Weapon, sortPriority=200}
  n34 = MaterialFunctionCall(Float4=@n5) {materialFunction=&BreakOutFloat4Components}
  n35 = LinearInterpolate(
   A=ScalarParameter() {defaultValue=150, group=Effect_SuperLevelling, parameterName=SL_Noise_TileRate_Vehicle, sortPriority=200}
   B=ScalarParameter() {defaultValue=400, group=Effect_SuperLevelling, parameterName=SL_Noise_TileRate_VehicleSide, sortPriority=200}
   Alpha=@n34.G
  )
  n36 = ScalarParameter() {defaultValue=300, group=Effect_SuperLevelling, parameterName=SL_Noise_TileRate_Mech, sortPriority=200}
  n37 = ScalarParameter() {defaultValue=40, group=Effect_SuperLevelling, parameterName=SL_Noise_TileRate_Character, sortPriority=200}
  n38 = PreSkinnedPosition()
  n39 = MaterialFunctionCall(
   TextureObject=@n32
   TextureSize=LinearInterpolate(A=LinearInterpolate(A=@n33, B=@n35, Alpha=@n34.R), B=LinearInterpolate(A=@n36, B=@n37, Alpha=@n34.A), Alpha=@n34.B)
   WorldPosition=FunctionInput(Preview=VertexInterpolator(VS=@n38).PS) {bUsePreviewValueAsDefault=true, inputName=Local Position}
   World Space Normal=@n8
  ) {materialFunction=&/WorldAlignedTexture}
  n40 = ScalarParameter() {defaultValue=0.5, group=Effect_SuperLevelling, parameterName=SL_Transition_Falloff_Start}
  n41 = ScalarParameter() {defaultValue=0.1, group=Effect_SuperLevelling, parameterName=SL_Transition_Falloff_End}
  n42 = Power(Base=@n30.R)
  n43 = Subtract(
   A=@n31
   B=Multiply(
    A=ChannelMaskParameter(None=@n39.XYZ Texture) {group=Effect_SuperLevelling, parameterName=SL_TransitionNoiseChannel}.R
    B=LinearInterpolate(A=@n40, B=@n41, Alpha=@n42)
   )
  )
  n44 = Distance(A=@n43, B=ScalarParameter() {defaultValue=0.5, group=Effect_SuperLevelling, parameterName=SL_EdgeSamplePosition})
  n45 = ScalarParameter() {defaultValue=0.05, group=Effect_SuperLevelling, parameterName=SL_EdgeWidth}
  n46 = Clamp(None=OneMinus(None=Divide(A=@n44, B=@n45)))
  n47 = Multiply(
   A=Multiply(A=VectorParameter() {group=Effect_SuperLevelling, parameterName=SL_EdgeGlowColor, defaultValue=<rgba 1 0.4812 0 1>}.RGB, B=@n46)
   B=ScalarParameter() {defaultValue=5, group=Effect_SuperLevelling, parameterName=SL_EdgeGlowStrength}
  )
  n48 = ScalarParameter() {defaultValue=2, group=Effect_SuperLevelling, parameterName=SL_EdgeGlowPower}
  n49 = Subtract(A=Power(Base=@n20, Exp=ScalarParameter() {defaultValue=1.33, group=Effect_SuperLevelling, parameterName=SL_Animation_EaseIn, sortPriority=103}) {constExponent=1.33}) {constB=0.5}
  n50 = OneMinus(None=Abs(None=Multiply(A=@n49) {constB=2}))
  n51 = ScalarParameter() {defaultValue=6, group=Effect_SuperLevelling, parameterName=SL_Animation_EaseInOut, sortPriority=103}
  n52 = Clamp(None=StaticSwitchParameter(True=Round(None=@n43), False=@n43) {defaultValue=true, group=Effect_SuperLevelling, parameterName=SL_UseHardEdgedMask})
  n53 = LinearInterpolate(A=@n1.EmissiveColor, B=Multiply(A=Power(Base=@n47, Exp=@n48), B=Power(Base=@n50, Exp=@n51) {constExponent=8}), Alpha=@n52)
  n54 = GetMaterialAttributes(None=SetMaterialAttributes!(
    MaterialAttributes=@n1.MaterialAttributes
    EmissiveColor=StaticSwitchParameter(True=@n53, False=@n1.EmissiveColor) {group=Effect_SuperLevelling, parameterName=Effect_EnableEdgeGlow, sortPriority=0}
   ) {attributeSetTypes [#1]}) {attributeGetTypes [#2, #1]}
  n55 = VectorParameter() {group=Effect_SuperLevelling_SimpleMaterial, parameterName=SM_BaseColor, defaultValue=<rgba 0 0 0 1>}
  n56 = ScalarParameter() {defaultValue=0.1, group=Effect_SuperLevelling_SimpleMaterial, parameterName=SM_Roughness}
  n57 = Constant3Vector()
  n58 = MaterialFunctionCall(X=@n52, Y=@n46, Z=Constant()) {materialFunction=&MakeFloat3}
  n59 = ChannelMaskParameter(None=@n58.Result) {group=Effect_SuperLevelling, maskChannel=Green, parameterName=SM_SuperLevellingEdgeChannel, defaultValue=<rgba 0 1 0 0>}
  n60 = SetMaterialAttributes!(
   MaterialAttributes=@n54.MaterialAttributes
   BaseColor=@n55.RGB
   Metallic=ScalarParameter() {defaultValue=1, group=Effect_SuperLevelling_SimpleMaterial, parameterName=SM_Metallic}
   Roughness=@n56
   Normal=@n54.MaterialAttributes
   Specular=ScalarParameter() {defaultValue=2, group=Effect_SuperLevelling_SimpleMaterial, parameterName=SM_Specular}
   SubsurfaceColor=@n57
   EmissiveColor=LinearInterpolate(A=@n57, B=@n54.EmissiveColor, Alpha=@n59.R)
  ) {attributeSetTypes [
    #3
    #4
    #5
    #2
    #6
    #7
    #1
   ]}
  n61 = ChannelMaskParameter(None=@n58.Result) {group=Effect_SuperLevelling, parameterName=SM_SuperLEvellingMaskChannel}
  FunctionOutput(None=StaticSwitchParameter(True=BlendMaterialAttributes(A=@n54.MaterialAttributes, B=@n60, Alpha=@n61.R) {vertexAttributeBlendType=UseA}, False=@n1.MaterialAttributes) {group=Customization, parameterName=Effect_SuperLeveling})
  MaterialFunctionCall(
   A=VectorParameter() {group=Effect_SuperLevelling, parameterName=SL_MaskPos_01}.RGB
   B=VectorParameter() {group=Effect_SuperLevelling, parameterName=SL_MaskPos_02}.RGB
   C=VectorParameter() {group=Effect_SuperLevelling, parameterName=SL_MaskPos_03}.RGB
   Alpha=@n42
  ) {materialFunction=&Lerp_3Color}
  Multiply(A=@n31)
  n62 = TextureSampleParameter2D() {
   group=Effect_SuperLevelling_SimpleMaterial
   parameterName="SM_BaseColor,Roughness(RGB,A)"
   texture=/Game/Weapons/WeaponSkins/Wraps/Textures/CatBurglarGameplay/T_Wrap_CatBurglarGameplay_D
  }
  StaticSwitchParameter(True=@n62.RGB, False=@n55.RGB) {group=Effect_SuperLevelling_SimpleMAterial, parameterName=SM_UseTexture}
  StaticSwitchParameter(
   True=LinearInterpolate(
    A=ScalarParameter() {group=Effect_SuperLevelling_SimpleMaterial, parameterName=SM_MinRoughness}
    B=ScalarParameter() {defaultValue=1, group=Effect_SuperLevelling_SimpleMaterial, parameterName=SM_MaxRoughness}
    Alpha=@n62.A
   )
   False=@n56
  ) {group=Effect_SuperLevelling_SimpleMAterial, parameterName=SM_UseTexture}
  VectorParameter() {group=Effect_SuperLevelling, parameterName=SL_TransitionNoise_Tiling, defaultValue=<rgba 5 5 5 1>}
 ]}
