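# Weapon wrap material function that adds an iridescent thin-film effect to input material attributes, gated by a static switch.
# Fresnel, state-dependent scale and an optional animated offset index a spectrum texture or color-curve atlas.
# That color is added to base color or emissive, weighted by inverse roughness and Fresnel.
# Primary/accent masks, top-side layer masks and decal masks, or optionally a metallic-based mask, control the final blend.
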MaterialFunction /Game/Weapons/WeaponSkins/Wraps/MaterialFunctions/MF_Effect_ThinFilm
graph {nodes [
  Constant2Vector()
  n1 = NamedRerouteUsage()
  n2 = NamedRerouteUsage()
  n3 = ScalarParameter() {defaultValue=1, group="Effect-ThinFilm  Msk", parameterName="ECTFilm_Primary   0=Off", sortPriority=0}
  n4 = TextureSample() {samplerType=SAMPLERTYPE_Masks, texture=/Game/Weapons/FORT_Rifles/Textures/T_Scar_CM}
  n5 = MaterialFunctionCall(Float4=FunctionInput(Preview=@n4.RGBA) {inputName=CM_Msk, inputType=FunctionInput_Vector4, sortPriority=4}) {materialFunction=&BreakOutFloat4Components}
  n6 = NamedRerouteDeclaration(None=@n5.R) {name=Msk_Prime, nodeColor=<rgba 0 1 0.7422 1>}
  n7 = Min(A=@n3, B=@n6)
  n8 = Min(
   A=NamedRerouteDeclaration(None=@n5.B) {name=Msk_Accent, nodeColor=<rgba 0 1 0.7656 1>}
   B=ScalarParameter() {group="Effect-ThinFilm  Msk", parameterName="ECTFilm_Accent    0=Off", sortPriority=0}
  )
  n9 = TextureSample() {samplerType=SAMPLERTYPE_Normal, texture=/Game/Weapons/FORT_Rifles/T_SCAR_N}
  n10 = Constant3Vector() {constant=<rgba 1 0 0 0>}
  n11 = GetMaterialAttributes(None=FunctionInput(Preview=SetMaterialAttributes!(MaterialAttributes, Normal=@n9.RGB, BaseColor=@n10) {attributeSetTypes [#1, #2]}) {inputName=Material, inputType=FunctionInput_MaterialAttributes}) {attributeGetTypes [
    #2
    #3
    #1
    #4
   ]}
  n12 = LinearInterpolate(
   A=ScalarParameter() {group="Effect-ThinFilm  Msk", parameterName=ECTFilm_Metalic Low}
   B=ScalarParameter() {defaultValue=1, group="Effect-ThinFilm  Msk", parameterName=ECTFilm_Metalic High}
   Alpha=@n11.Metallic
  )
  n13 = FunctionInput() {inputName=Decal_Mask, inputType=FunctionInput_Scalar, sortPriority=90}
  n14 = FunctionInput() {inputName=Decal_Mask_RGBA, inputType=FunctionInput_Vector4, sortPriority=90}
  n15 = Constant()
  n16 = StaticSwitchParameter(
   True=@n13
   False=StaticSwitchParameter(True=ChannelMaskParameter(None=@n14) {group="Effect-ThinFilm  Msk", parameterName=ECTFilm_Decal_Mask_RGBA_Channel}.R, False=@n15) {group="Effect-ThinFilm  Msk", parameterName=ECTFilm_Decal_Mask_RGBA}
  ) {group="Effect-ThinFilm  Msk", parameterName=ECTFilm_Decal_Mask_Alpha}
  n17 = ScalarParameter() {defaultValue=1, group="Effect-ThinFilm  Msk", parameterName="ECTFilm_Decal_Alpha Masked  1=Active"}
  n18 = StaticSwitchParameter(True=OneMinus(None=LinearInterpolate(B=@n13, Alpha=@n17)), False=Constant() {r=1}) {group=Customization, parameterName=Function_Left_Right_Decal_V2}
  n19 = VectorParameter() {
   group="Effect-ThinFilm  Msk"
   parameterName=ECTFilm_Decal_Multi_Disable
   channelNames {a="DecalD     1-On", b="DecalC     1-On", g="DecalB     1-On", r="DecalA     1-On"}
  }
  MaterialExpressionFunctionInput_7 = FunctionInput(Preview=Constant4Vector().RGBA) {bUsePreviewValueAsDefault=true, inputName=multi_Decal, inputType=FunctionInput_Vector4}
  n20 = Convert(None=@MaterialExpressionFunctionInput_7) {
   nodeName=Break Float4
   convertInputs [{
     type=Vector4
     defaultValue=<rgba 0 0 0 1>
     expressionInput {
      expression=&self:MaterialExpressionFunctionInput_7
      inputName=None
      mask=0
      maskA=0
      maskB=0
      maskG=0
      maskR=0
      outputIndex=0
     }
    }]
   convertOutputs [{type=Scalar}, {type=Scalar}, {type=Scalar}, {type=Scalar}]
  }
  n21 = Max(
   A=Max(A=Min(A=@n19."DecalA     1-On", B=@n20), B=Min(A=@n19."DecalB     1-On", B=@n20))
   B=Max(A=Min(A=@n19."DecalC     1-On", B=@n20), B=Min(A=@n19."DecalD     1-On", B=@n20))
  )
  n22 = ChannelMaskParameter(None=FunctionInput() {inputName=TS_Layer01_Mask, inputType=FunctionInput_Vector4, sortPriority=99}) {group="Effect-ThinFilm  Msk", parameterName=ECTFilm_TS_Layer01_Mask_Channel}
  n23 = FunctionInput() {inputName=TS_Layer03_Mask, inputType=FunctionInput_Vector4, sortPriority=99}
  n24 = ChannelMaskParameter(None=FunctionInput() {inputName=TS_Layer04_Mask, inputType=FunctionInput_Vector4, sortPriority=99}) {group="Effect-ThinFilm  Msk", parameterName=ECTFilm_TS_Layer04_Mask_Channel}
  n25 = Constant() {r=1}
  n26 = StaticSwitchParameter(
   True=ChannelMaskParameter(None=FunctionInput() {inputName=TS_Layer02_Mask, inputType=FunctionInput_Vector4, sortPriority=99}) {group="Effect-ThinFilm  Msk", parameterName=ECTFilm_TS_Layer02_Mask_Channel}.R
   False=StaticSwitchParameter(
    True=ChannelMaskParameter(None=@n23) {group="Effect-ThinFilm  Msk", parameterName=ECTFilm_TS_Layer03_Mask_Channel}.R
    False=StaticSwitchParameter(True=@n24.R, False=@n25) {group="Effect-ThinFilm  Msk", parameterName=ECTFilm_TS_Layer04_Mask}
   ) {group="Effect-ThinFilm  Msk", parameterName=ECTFilm_TS_Layer03_Mask}
  ) {group="Effect-ThinFilm  Msk", parameterName=ECTFilm_TS_Layer02_Mask}
  n27 = Constant()
  n28 = NamedRerouteUsage()
  n29 = ScalarParameter() {group=TopSide_V02, parameterName="TopS_Mask_Accent  0=Off"}
  n30 = StaticSwitch(
   True=Max(A=LinearInterpolate(A=@n27, B=@n28, Alpha=@n29), B=ScalarParameter() {group=TopSide_V02, parameterName="TopS_Mask_Primary\\Accent  1=Off"})
   False=@n27
   Value=StaticBoolParameter() {group=TopSide_V02, parameterName=TopSideProjection_V02}
  )
  n31 = Min(
   A=StaticSwitch(
    True=StaticSwitchParameter(True=@n22.R, False=@n26) {group="Effect-ThinFilm  Msk", parameterName=ECTFilm_TS_Layer01_Mask}
    False=@n25
    Value=StaticBoolParameter() {group=TopSide_V02, parameterName=TopSideProjection_V02}
   )
   B=Min(A=@n3, B=Max(A=@n30, B=@n6))
  )
  n32 = Multiply(
   A=StaticSwitchParameter(True=OneMinus(None=@n21), False=Constant() {r=1}) {group=Customization, parameterName=Effect_Multi_Decal_Reactive}
   B=Max(A=Max(A=@n31, B=@n8), B=ScalarParameter() {group="Effect-ThinFilm  Msk", parameterName="ECTFilm_Mask_Disable 1=Active", sortPriority=1})
  )
  n33 = StaticSwitchParameter(
   True=Min(A=Max(A=@n7, B=@n8), B=Saturate(None=@n12))
   False=Max(A=StaticSwitchParameter(True=@n16, False=@n15) {group=Customization, parameterName=Function_Left_Right_Decal_V2}, B=Min(A=@n18, B=@n32))
  ) {group="Effect-ThinFilm  Msk", parameterName=ECTFilm_Metalic Channel Use}
  FunctionOutput(None=StaticSwitchParameter(
    True=BlendMaterialAttributes(A=@n1, B=@n2, Alpha=@n33) {vertexAttributeBlendType=UseA}
    False=NamedRerouteDeclaration(None=@n11.MaterialAttributes) {name=Root_Material, nodeColor=<rgba 0.7812 1 0 1>}
   ) {group=Customization, parameterName=Effect_ThinFilm}) {outputName=Result_1}
  NamedRerouteDeclaration(None=@n5.G) {name=Msk_Curvature, nodeColor=<rgba 0 1 0.7422 1>}
  n34 = GetMaterialAttributes(None=NamedRerouteUsage()) {attributeGetTypes [#3, #2, #5]}
  n35 = ScalarParameter() {defaultValue=98, group=Effect-ThinFilm Base, parameterName="ECTFilm_ColorCurve  Index"}
  n36 = ScalarParameter() {defaultValue=1, group=Effect-ThinFilm Base, parameterName=ECTFilm_Animated_TimeDivisor, sortPriority=50}
  n37 = ScalarParameter() {defaultValue=1, group=Effect-ThinFilm Base, parameterName=ECTFilm_Animated_TimeModulo, sortPriority=50}
  n38 = ScalarParameter() {defaultValue=1, group=Effect-ThinFilm Base, parameterName=ECTFilm_Scale, sortPriority=50}
  n39 = VectorParameter() {
   group=Effect-ThinFilm Base
   parameterName=ECTFilm_Scale_States
   sortPriority=50
   channelNames {a="", b="", g="Character 1=on", r=Character Scale}
  }
  n40 = Constant4Vector()
  n41 = FunctionInput(Preview=@n40.RGBA) {bUsePreviewValueAsDefault=true, inputName=IsState, inputType=FunctionInput_Vector4, sortPriority=4}
  n42 = LinearInterpolate(
   A=@n38
   B=@n39.Character Scale
   Alpha=Min(
    A=MaterialFunctionCall(Float4=@n41) {materialFunction=&BreakOutFloat4Components}.A
    B=@n39."Character 1=on"
   )
  )
  n43 = MaterialFunctionCall(Float4=NamedRerouteDeclaration(None=FunctionInput(Preview=@n40.RGBA) {bUsePreviewValueAsDefault=true, inputName=IsState_B, inputType=FunctionInput_Vector4, sortPriority=4}) {name=IsState_B, nodeColor=<rgba 0.1497 1 0.1633 1>}) {materialFunction=&BreakOutFloat4Components}
  n44 = ScalarParameter() {defaultValue=0.5, group=Effect-ThinFilm Base, parameterName=ECTFilm_Warp, sortPriority=50}
  n45 = Add(
   A=StaticSwitchParameter(
    True=MaterialFunctionCall!(Time, Divisor=@n36, Mod=@n37) {materialFunction=/Game/Effects/Fort_Effects/Materials/_Functions/MF_ModuloTime}.Unclamped
    False=ScalarParameter() {group=Effect-ThinFilm Base, parameterName=ECTFilm_Offset, sortPriority=50}
   ) {group=Effect-ThinFilm Base, parameterName=ECTFilm_Animate, sortPriority=50}
   B=Multiply(A=LinearInterpolate(A=@n42, B=@n38, Alpha=@n43.R), B=Fresnel(ExponentIn=@n44) {baseReflectFraction=0})
  )
  n46 = Frac(None=@n45)
  n47 = CurveAtlasRowParameter(CurveTime=@n46) {
   atlas=/Game/Weapons/WeaponSkins/Wraps/Textures/CruveAtlas_Wraps_01
   curve=/Game/Weapons/WeaponSkins/Wraps/Textures/0_ColorCurves/CC_Spectum
   defaultValue=98
   group=Effect-ThinFilm Base
   parameterName=ECTFilm_ColorCurve
   sortPriority=6
  }
  n48 = TextureSampleParameter2D(UVs=@n45) {
   group=Effect-ThinFilm Base
   parameterName=ECTFilm Texture
   sortPriority=6
   texture=/Game/Global/Textures/Default/Thinfilm/T_ThinFilm_Spectrum_COLOR
  }
  n49 = Multiply(
   A=StaticSwitchParameter(
    True=MaterialFunctionCall!(Index=@n35, CCinput=@n47, CCinput_Alpha=@n47, UV=@n46) {materialFunction=&/MF_Global_CC_Tex}.3V
    False=@n48.RGB
   ) {group=Effect-ThinFilm Base, parameterName="ECTFilm Texture---->CC", sortPriority=5}
   B=ScalarParameter() {defaultValue=0.5, group=Effect-ThinFilm Base, parameterName=ECTFilm_Intensity}
  )
  n50 = Saturate(None=Multiply(A=OneMinus(None=@n34.Roughness), B=ScalarParameter() {defaultValue=1, group=Effect-ThinFilm Base, parameterName=ECTFilm_Roughness-Scale}))
  n51 = ScalarParameter() {group=Effect-ThinFilm Base, parameterName="ECTFilm_Roughness-Scale  1-disable"}
  n52 = ScalarParameter() {defaultValue=1.2164, group=Effect-ThinFilm Base, parameterName=ECTFilm_Roughness_Exponent}
  n53 = Max(
   A=Multiply(A=Max(A=@n50, B=@n51), B=Fresnel(ExponentIn=@n52) {baseReflectFraction=0})
   B=ScalarParameter() {group=Effect-ThinFilm Base, parameterName="ECTFilm_Roughness Disabled    1=On"}
  )
  n54 = LinearInterpolate(A=@n34.BaseColor, B=Add(A=@n34.BaseColor, B=@n49), Alpha=@n53)
  n55 = ScalarParameter() {group=Effect-ThinFilm Base, parameterName="ECTFilm_BaseColor----->Emissive", sortPriority=2}
  n56 = LinearInterpolate(A=@n34.EmissiveColor, B=Add(A=@n34.EmissiveColor, B=@n49), Alpha=@n53)
  NamedRerouteDeclaration(None=SetMaterialAttributes!(
    MaterialAttributes=@n34.MaterialAttributes
    BaseColor=LinearInterpolate(A=@n54, B=@n34.BaseColor, Alpha=@n55)
    EmissiveColor=LinearInterpolate(A=@n34.EmissiveColor, B=@n56, Alpha=@n55)
   ) {attributeSetTypes [#2, #5]}) {name=Material, nodeColor=<rgba 0 0.5078 1 1>}
  NamedRerouteDeclaration(None=@n4.RGBA) {name=CM_Mask, nodeColor=<rgba 1 0 0.3047 1>}
  NamedRerouteDeclaration(None=@n41) {name=IsState, nodeColor=<rgba 0.1497 1 0.1633 1>}
  n57 = PreSkinnedLocalBounds()
  n58 = Subtract(A=VertexInterpolator(VS=PreSkinnedPosition()).PS, B=@n57.Min)
  n59 = MaterialFunctionCall(Color=Subtract(A=@n57.Max, B=@n57.Max)) {materialFunction=&SplitComponents}
  n60 = Max(A=Max(A=@n59.R, B=@n59.R), B=@n59.B)
  NamedRerouteDeclaration(None=FunctionInput(Preview=Divide(A=@n58, B=@n60)) {bUsePreviewValueAsDefault=true, inputName=Calcated_Bounds, sortPriority=11}) {name=Caclcated_Bounds, nodeColor=<rgba 0 0.9297 1 1>}
  n61 = PreSkinnedNormal()
  NamedRerouteDeclaration(None=FunctionInput(Preview=VertexInterpolator(VS=@n61).PS) {bUsePreviewValueAsDefault=true, inputName=Pre-Skinned Local Normal, sortPriority=10}) {name=PreSkin_Normal, nodeColor=<rgba 0.8281 1 0 1>}
  NamedRerouteDeclaration(None=@n11.Normal) {name=Root_Normal, nodeColor=<rgba 0.4062 0 1 1>}
 ]}
