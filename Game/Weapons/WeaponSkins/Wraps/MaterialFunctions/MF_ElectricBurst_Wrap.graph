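# Material function for the Electric Burst wrap on weapons, vehicles and mechs, gated by a static switch.
# Top/side projection over wrap bounds samples base color and panning masks, height-lerped against animated gradients and combined with edge gradients.
# Time since a Blueprint-fed game time drives curved fades of base color and boosted emissive, creating a reactive burst.
# Customization masks blend the result over incoming material attributes.
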MaterialFunction /Game/Weapons/WeaponSkins/Wraps/MaterialFunctions/MF_ElectricBurst_Wrap
graph {nodes [
  n1 = VectorParameter() {group=TopSideProjection, parameterName=TS_Fill_Z_Axis_Weapon_Color_Mask_01, sortPriority=21, defaultValue=<rgba 0 0 0 1>}
  AppendVector(A=@n1.RGB, B=@n1.RGB)
  n2 = VectorParameter() {group=Emissive_Curvture_Mask, parameterName=Mask_WeaponOffset_ CE, sortPriority=161, defaultValue=<rgba 0 -1 0 1>}
  n3 = VectorParameter() {group=Emissive_Curvture_Mask, parameterName=Mask_VehicleOffset_ CE, sortPriority=163, defaultValue=<rgba -1 0 0 1>}
  n4 = VectorParameter() {group=Emissive_Curvture_Mask, parameterName=Mask_MechOffset_ CE, sortPriority=165, defaultValue=<rgba 0 0 -1 1>}
  n5 = Constant3Vector() {constant=<rgba 1 0 0 0>}
  n6 = FunctionInput(Preview=Constant()) {bUsePreviewValueAsDefault=true, inputName=isMech, inputType=FunctionInput_Scalar, sortPriority=4}
  n7 = FunctionInput(Preview=Constant()) {inputName=isVehicle, inputType=FunctionInput_Scalar, sortPriority=5}
  n8 = VectorParameter() {group=Emissive_Curvture_Mask, parameterName=Mask_UV_Spread_XY CE, sortPriority=159, defaultValue=<rgba 1 1 1 0>}
  n9 = VectorParameter() {group=Emissive_Curvture_Mask, parameterName=Mask_UV_Spread_Mech_XY CE, sortPriority=159, defaultValue=<rgba 1 1 1 0>}
  n10 = PreSkinnedPosition()
  n11 = MaterialFunctionCall!(
   is Vehicle=@n7
   Object Bounds Max Adjustment=FunctionInput(Preview=Constant3Vector() {constant=<rgba 0 -0.5 0 1>}) {bUsePreviewValueAsDefault=true, inputName=Object Bounds Max Adjustment, sortPriority=13}
   Object Bounds Min Adjustment=FunctionInput(Preview=Constant3Vector() {constant=<rgba 0 31.5 0 1>}) {bUsePreviewValueAsDefault=true, inputName=Object Bounds Min Adjustment, sortPriority=12}
   Object Bounds Center Adjustment=FunctionInput(Preview=Constant3Vector() {constant=<rgba 0.2122 0 -0.09 0>}) {bUsePreviewValueAsDefault=true, inputName=Object Bounds Center Adjustment, sortPriority=14}
   isMech=@n6
   Side Project Vehicle
   Local Position=FunctionInput(Preview=VertexInterpolator(VS=@n10).PS) {bUsePreviewValueAsDefault=true, inputName=Local Position, sortPriority=11}
  ) {materialFunction=&/MF_WrapBounds}
  n12 = TextureSampleParameter2D(UVs=MaterialFunctionCall!(
    ForwardAxis=FunctionInput(Preview=@n5) {bUsePreviewValueAsDefault=true, inputName=ForwardAxis}
    Bounds=@n11.Result
    Pre-Skinned Local Position
    isVehicle=@n7
    isMech=@n6
    Side Project Vehicle
    WeaponOffset=@n2.RGB
    WeaponFalloff=ScalarParameter() {defaultValue=0.5, group=Emissive_Curvture_Mask, parameterName=Mask_WeaponFalloff CE, sortPriority=162}
    VehicleOffset=@n3.RGB
    VehicleFalloff=ScalarParameter() {defaultValue=0.5, group=Emissive_Curvture_Mask, parameterName=Mask_VehicleFalloff CE, sortPriority=164}
    MechOffset=@n4.RGB
    MechFalloff=ScalarParameter() {group=Emissive_Curvture_Mask, parameterName=Mask_MechFalloff CE, sortPriority=166}
    Invert_Gradent
    UV_Spread_XY=LinearInterpolate(A=@n8.RGB, B=@n9.RGB, Alpha=@n6)
    UV_Animated_XY=VectorParameter() {group=Emissive_Curvture_Mask, parameterName=Mask_UV_Animated_XY CE, sortPriority=160}.RGB
    Warp_UVs=Constant()
    Rotate_Mech
    Rotate_Vehicle_Side
    Rotate_Vehicle
   ) {materialFunction=&/MF_Wrap_Gradent_Creation}.Uvs) {
   group=Emissive_Curvture_Mask
   parameterName=Mask_Gradent_Texture CE
   samplerType=SAMPLERTYPE_Masks
   sortPriority=155
   texture=/Game/Global/Textures/Default/Masks_and_Gradients/T_VertGradient_MASK
  }
  ChannelMaskParameter(None=@n12.RGB) {group=Emissive_Curvture_Mask, parameterName=Mask_Gradent_Texture_Channel CE, sortPriority=154}
  n13 = TextureObjectParameter() {group=TopSideProjection, parameterName=TS_BaseColor_01, sortPriority=11, texture=&T_BlacksRGB}
  n14 = ScalarParameter() {defaultValue=1, group=TopSideProjection, parameterName=TS_Blend_Contrast, sortPriority=-1}
  n15 = LinearInterpolate(A=@n14, B=ScalarParameter() {defaultValue=2, group=TopSideProjection, parameterName=TS_Blend_Contrast_Mech, sortPriority=-1}, Alpha=@n6)
  n16 = Constant()
  n17 = VectorParameter() {group=TopSideProjection, parameterName=TS_Fill_alternate_Axis_Color_Mask_01, sortPriority=21, defaultValue=<rgba 0 0 0 1>}
  n18 = AppendVector(A=@n17.RGB, B=@n17.RGB)
  n19 = VectorParameter() {group=TopSideProjection_UVS, parameterName=TS-Rotation-Weapon-01---------------------, sortPriority=1, defaultValue=<rgba 0 0 -0.25 1>}
  n20 = VectorParameter() {group=TopSideProjection_UVS, parameterName=TS_Rotation_Vehicle_01, sortPriority=2, defaultValue=<rgba 0 0 0 1>}
  n21 = VectorParameter() {group=TopSideProjection_UVS, parameterName=TS_Rotation_Mech_01, sortPriority=3, defaultValue=<rgba 0.25 0.25 0 1>}
  n22 = LinearInterpolate(A=LinearInterpolate(A=@n19.RGB, B=@n20.RGB, Alpha=@n7), B=@n21.RGB, Alpha=@n6)
  n23 = FunctionInput(Preview=VertexInterpolator(VS=PreSkinnedNormal()).PS) {bUsePreviewValueAsDefault=true, inputName=Pre-Skinned Local Normal, sortPriority=10}
  n24 = MaterialFunctionCall!(
   IsCharacter
   IsStatsB
   Texture=@n13
   Rotation_Axis=@n22
   Caculated_Bounds=@n11.Result
   Pre_Skinned_Normals=@n23
   Blend_Contrast=@n15
   Fill_alternate_Axis_Opacity=@n16
   "Fill_alternate_Axis_Color_A=Roughness"=@n18
   Is_Vehicle=@n7
   Is_Mech=@n6
   Enable_Fill_Z_Weapon
   Fill_Z_Weapon_opacity
   Fill_Z_Weapon_Color
  ) {materialFunction=&/MF_TopsSideProjection_Base}
  n25 = ComponentMask(None=@n24.Combined) {b=true, g=true, r=true}
  n26 = MaterialFunctionCall(Color=Multiply(A=@n11.Result, B=Constant3Vector() {constant=<rgba 1 1 -1 0>})) {materialFunction=&SplitComponents}
  n27 = OneMinus(None=@n26.B)
  n28 = LinearInterpolate(
   A=LinearInterpolate(A=AppendVector(A=@n26.G, B=@n26.G), B=AppendVector(A=@n26.R, B=@n26.R), Alpha=@n7)
   B=Multiply(A=AppendVector(A=@n27, B=@n26.R)) {constB=0.425}
   Alpha=@n6
  )
  n29 = Power(Base=Multiply(A=MaterialFunctionCall(UV Channel=@n28) {materialFunction=&LinearGradient}.UGradient) {constB=0.975}) {constExponent=10}
  n30 = VectorParameter() {group=TopSideProjection_UVS, parameterName=TS_BaseColor_MaskColor, sortPriority=-6, defaultValue=<rgba 1 1 1 0>}
  n31 = MaterialFunctionCall(UV Channel=@n28) {materialFunction=&LinearGradient}
  n32 = Power(Base=Multiply(A=OneMinus(None=@n31.UGradient)) {constB=0.975}) {constExponent=10}
  n33 = Power(Base=Add(A=Multiply(A=@n29, B=@n30.RGB), B=Multiply(A=@n32, B=@n30.RGB))) {constExponent=1.5}
  n34 = TextureObjectParameter() {
   group=Electric Burst
   parameterName=EB_BaseMask_01
   samplerType=SAMPLERTYPE_Masks
   sortPriority=20
   texture=&T_BlackMask
  }
  n35 = Time()
  n36 = VectorParameter() {group=Electric Burst, parameterName=EB_PanningBCMask_Weapon_02, sortPriority=10, defaultValue=<rgba 0 0 0 1>}
  n37 = VectorParameter() {group=Electric Burst, parameterName=EB_PanningBCMask_Vehicle_02, sortPriority=11, defaultValue=<rgba 0 0 0 1>}
  n38 = VectorParameter() {group=Electric Burst, parameterName=EB_PanningBCMask_Mech_02, sortPriority=12, defaultValue=<rgba 0 0 0 1>}
  n39 = LinearInterpolate(A=LinearInterpolate(A=@n36.RGB, B=@n37.RGB, Alpha=@n7), B=@n38.RGB, Alpha=@n6)
  n40 = VectorParameter() {group=TopSideProjection, parameterName="TS_Fill_alternate_Axis_Color_A=Roughness 01", sortPriority=3}
  n41 = AppendVector(A=@n40.RGB, B=@n40.RGB)
  n42 = StaticBool() {value=true}
  n43 = MaterialFunctionCall(
   Texture=@n34
   Rotation_Axis=@n22
   Caculated_Bounds=Add(A=@n11.Result, B=Multiply(A=@n35, B=@n39))
   Pre_Skinned_Normals=@n23
   Blend_Contrast=@n15
   Fill_alternate_Axis_Opacity=@n16
   "Fill_alternate_Axis_Color_A=Roughness"=@n41
   Is_Vehicle=@n7
   Is_Mech=@n6
   Enable_Fill_Z_Weapon=@n42
  ) {materialFunction=&MF_TopsSideProjection_Base}
  n44 = Time()
  n45 = VectorParameter() {group=Electric Burst, parameterName=EB_PanningBCMask_Weapon_01, sortPriority=-4, defaultValue=<rgba 0 0 0 1>}
  n46 = VectorParameter() {group=Electric Burst, parameterName=EB_PanningBCMask_Vehicle_01, sortPriority=-3, defaultValue=<rgba 0 0 0 1>}
  n47 = VectorParameter() {group=Electric Burst, parameterName=EB_PanningBCMask_Mech_01, sortPriority=-2, defaultValue=<rgba 0 0 0 1>}
  n48 = LinearInterpolate(A=LinearInterpolate(A=@n45.RGB, B=@n46.RGB, Alpha=@n7), B=@n47.RGB, Alpha=@n6)
  n49 = MaterialFunctionCall(
   Texture=@n34
   Rotation_Axis=@n22
   Caculated_Bounds=Add(A=@n11.Result, B=Multiply(A=@n44, B=@n48))
   Pre_Skinned_Normals=@n23
   Blend_Contrast=@n14
   Fill_alternate_Axis_Opacity=@n16
   "Fill_alternate_Axis_Color_A=Roughness"=@n41
   Is_Vehicle=@n7
   Is_Mech=@n6
   Enable_Fill_Z_Weapon=@n42
  ) {materialFunction=&MF_TopsSideProjection_Base}
  n50 = VectorParameter() {group=Electric Burst, parameterName=EB_Mask_WeaponOffset_ CE, sortPriority=180, defaultValue=<rgba 0 -1 0 1>}
  n51 = VectorParameter() {group=Electric Burst, parameterName=EB_Mask_VehicleOffset_ CE, sortPriority=181, defaultValue=<rgba -1 0 0 1>}
  n52 = VectorParameter() {group=Electric Burst, parameterName=EB_Mask_MechOffset_ CE, sortPriority=182, defaultValue=<rgba 0 0 -1 1>}
  n53 = VectorParameter() {group=Electric Burst, parameterName=EB_Mask_UV_Spread_XY CE, sortPriority=183, defaultValue=<rgba 1 1 1 0>}
  n54 = VectorParameter() {group=Electric Burst, parameterName=EB_Mask_UV_Spread_Mech_XY CE, sortPriority=184, defaultValue=<rgba 1 1 1 0>}
  n55 = LinearInterpolate(
   A=VectorParameter() {group=Electric Burst, parameterName=EB_Mask_UV_AnimatedWeapon_XY CE, sortPriority=185}.RGB
   B=VectorParameter() {group=Electric Burst, parameterName=EB_Mask_UV_AnimatedVehicle_XY CE, sortPriority=186}.RGB
   Alpha=@n7
  )
  n56 = VectorParameter() {group=Electric Burst, parameterName=EB_Mask_UV_AnimatedMech_XY CE, sortPriority=187}
  n57 = TextureSample(
   UVs=MaterialFunctionCall(
    Bounds=@n11.Result
    isVehicle=@n7
    isMech=@n6
    WeaponOffset=@n50.RGB
    WeaponFalloff=ScalarParameter() {defaultValue=0.5, group=Electric Burst, parameterName=EB_Mask_WeaponFalloff CE, sortPriority=180}
    VehicleOffset=@n51.RGB
    VehicleFalloff=ScalarParameter() {defaultValue=0.5, group=Electric Burst, parameterName=EB_Mask_VehicleFalloff CE, sortPriority=181}
    MechOffset=@n52.RGB
    MechFalloff=ScalarParameter() {group=Electric Burst, parameterName=EB_Mask_MechFalloff CE, sortPriority=182}
    UV_Spread_XY=LinearInterpolate(A=@n53.RGB, B=@n54.RGB, Alpha=@n6)
    UV_Animated_XY=LinearInterpolate(A=@n55, B=@n56.RGB, Alpha=@n6)
    Warp_UVs=Constant() {r=0.1}
   ) {materialFunction=&MF_Wrap_Gradent_Creation}.Uvs
   Tex=@n34
  )
  n58 = Multiply(
   A=Clamp(
    None=ChannelMaskParameter(None=@n57.RGBA) {group=Electric Burst, parameterName=EB_BaseColor_Gradient_TextureChannel, sortPriority=120}.R
    Max=ScalarParameter() {defaultValue=1, group=Electric Burst, parameterName=EB_Mask_Gradent_Clamp Max, sortPriority=151}
   ) {maxDefault=0.99, minDefault=0.01}
   B=ScalarParameter() {defaultValue=1, group=Electric Burst, parameterName=EB_Mask_Gradent_Opacity_CE, sortPriority=149}
  )
  n59 = ScalarParameter() {group=TopSideProjection, parameterName=TS_Fill_alternate_Axis_Opacity 01, sortPriority=2}
  n60 = Time()
  n61 = VectorParameter() {group=Electric Burst, parameterName=EB_PanningBaseColor_Weapon_01, sortPriority=100, defaultValue=<rgba 0 0 0 1>}
  n62 = VectorParameter() {group=Electric Burst, parameterName=EB_PanningBaseColor_Vehicle_01, sortPriority=101, defaultValue=<rgba 0 0 0 1>}
  n63 = VectorParameter() {group=Electric Burst, parameterName=EB_PanningBaseColor_Mech_01, sortPriority=102, defaultValue=<rgba 0 0 0 1>}
  n64 = LinearInterpolate(A=LinearInterpolate(A=@n61.RGB, B=@n62.RGB, Alpha=@n7), B=@n63.RGB, Alpha=@n6)
  n65 = MaterialFunctionCall(
   Texture=@n34
   Rotation_Axis=@n22
   Caculated_Bounds=Add(A=@n11.Result, B=Multiply(A=@n60, B=@n64))
   Pre_Skinned_Normals=@n23
   Blend_Contrast=@n15
   Fill_alternate_Axis_Opacity=@n59
   "Fill_alternate_Axis_Color_A=Roughness"=@n41
   Is_Vehicle=@n7
   Is_Mech=@n6
   Enable_Fill_Z_Weapon=@n42
  ) {materialFunction=&MF_TopsSideProjection_Base}
  n66 = Saturate(None=MaterialFunctionCall(
    Transition Phase=Saturate(None=@n58)
    Height Texture=ChannelMaskParameter(None=@n65.Combined) {group=Electric Burst, parameterName=EB_BaseColor_TextureChannel, sortPriority=119}.R
    Contrast=ScalarParameter() {defaultValue=1, group=Electric Burst, parameterName=EB_Mask_HeightLerp_Contrast, sortPriority=145}
   ) {materialFunction=&HeightLerp}.Alpha)
  n67 = Add(
   A=Add(
    A=ChannelMaskParameter(None=@n43.Combined) {group=Electric Burst, parameterName=EB_BCMask02_TextureChannel, sortPriority=-5}.R
    B=ChannelMaskParameter(None=@n49.Combined) {group=Electric Burst, parameterName=EB_BCMask_TextureChannel, sortPriority=-5}.R
   )
   B=@n66
  )
  n68 = Add(A=@n25, B=Add(A=@n33, B=Multiply(A=@n30.RGB, B=@n67)))
  n69 = Subtract(A=Time(), B=ScalarParameter() {group=00-Reactive-GTFB, parameterName=00-GameTimeFromBP, sortPriority=-100})
  n70 = Divide(A=@n69, B=ScalarParameter() {defaultValue=1.5, group=Electric Burst, parameterName=EB_ReactiveDuration_AmbientArc, sortPriority=-99})
  n71 = OneMinus(None=Power(
    Base=Saturate(None=@n70)
    Exp=ScalarParameter() {defaultValue=1.4, group=Electric Burst, parameterName=EB_CurveAmount_AmbientArc, sortPriority=-98}
   ) {constExponent=0.25})
  n72 = FunctionInput() {bUsePreviewValueAsDefault=true, inputName="Customization Mask (_CM B Top left)", inputType=FunctionInput_Scalar, sortPriority=24}
  n73 = Add(A=OneMinus(None=@n72), B=@n25)
  n74 = Divide(A=@n69, B=ScalarParameter() {defaultValue=1.5, group=Electric Burst, parameterName=EB_ReactiveDuration, sortPriority=-100})
  n75 = OneMinus(None=Power(Base=Saturate(None=@n74), Exp=ScalarParameter() {defaultValue=1.2, group=Electric Burst, parameterName=EB_Curve_Amount, sortPriority=-100}))
  n76 = Saturate(None=Divide(A=@n69, B=Constant() {r=5}))
  n77 = VectorParameter() {group=Electric Burst, parameterName=TS_Reactive_EmissiveBoost, sortPriority=200, defaultValue=<rgba 4 4 4 0>}
  n78 = ScalarParameter() {defaultValue=1, group=Emissive_Curvture_Mask, parameterName=Mask_Gradent_Clamp Max, sortPriority=151}
  n79 = Saturate(None=Multiply(
    A=Clamp(None=@n12.B, Max=@n78) {maxDefault=0.99, minDefault=0.01}
    B=ScalarParameter() {defaultValue=1, group=Emissive_Curvture_Mask, parameterName=Mask_Gradent_Opacity_CE, sortPriority=149}
   ))
  n80 = TextureObjectParameter() {
   group=TopSideProjection
   parameterName=TS_Mask_01
   samplerType=SAMPLERTYPE_Masks
   sortPriority=20
   texture=&T_BlackMask
  }
  n81 = VectorParameter() {group=TopSideProjection_UVS, parameterName=TS_Panning_Weapon_01, sortPriority=20, defaultValue=<rgba 0 0 0 1>}
  n82 = VectorParameter() {group=TopSideProjection_UVS, parameterName=TS_Panning_Vehicle_01, sortPriority=21, defaultValue=<rgba 0 0 0 1>}
  n83 = VectorParameter() {group=TopSideProjection_UVS, parameterName=TS_Panning_Mech_01, sortPriority=22, defaultValue=<rgba 0 0 0 1>}
  n84 = Multiply(A=Time(), B=LinearInterpolate(A=LinearInterpolate(A=@n81.RGB, B=@n82.RGB, Alpha=@n7), B=@n83.RGB, Alpha=@n6))
  n85 = ChannelMaskParameter(None=MaterialFunctionCall(
    Texture=@n80
    Rotation_Axis=@n22
    Caculated_Bounds=Add(A=@n11.Result, B=@n84)
    Pre_Skinned_Normals=@n23
    Blend_Contrast=@n15
    Fill_alternate_Axis_Opacity=@n59
    "Fill_alternate_Axis_Color_A=Roughness"=@n18
    Is_Vehicle=@n7
    Is_Mech=@n6
   ) {materialFunction=&MF_TopsSideProjection_Base}.Combined) {group=TopSideProjection_Emissive, parameterName=Ts_Emissive01_TextureChannel, sortPriority=11}
  n86 = ScalarParameter() {defaultValue=1, group=Emissive_Curvture_Mask, parameterName=Mask_HeightLerp_Contrast, sortPriority=145}
  n87 = Multiply(
   A=@n30.RGB
   B=Saturate(None=MaterialFunctionCall(Transition Phase=@n79, Height Texture=@n85.R, Contrast=@n86) {materialFunction=&HeightLerp}.Alpha)
  )
  n88 = Add(A=@n33, B=Multiply(A=@n77.RGB, B=Add(A=@n66, B=@n87)))
  n89 = FunctionInput() {inputName=Material, inputType=FunctionInput_MaterialAttributes}
  n90 = GetMaterialAttributes(None=@n89) {attributeGetTypes [#1]}
  n91 = SetMaterialAttributes!(
   MaterialAttributes=@n90.MaterialAttributes
   BaseColor=LinearInterpolate(A=LinearInterpolate(A=@n68, B=@n25, Alpha=@n71), B=LinearInterpolate(A=@n25, B=@n73, Alpha=@n75), Alpha=OneMinus(None=@n76))
   Metallic=ScalarParameter() {group=TopSideProjection, parameterName=TS_Metallic_01, sortPriority=26}
   Specular=ScalarParameter() {defaultValue=0.5, group=TopSideProjection, parameterName=TS_Specularty, sortPriority=12}
   Roughness=LinearInterpolate(
    A=ScalarParameter() {group=TopSideProjection, parameterName=TS_Roughness_Min, sortPriority=13}
    B=ScalarParameter() {defaultValue=1, group=TopSideProjection, parameterName=TS_Roughness_Max, sortPriority=14}
    Alpha=ComponentMask(None=@n24.Alpha) {a=true}
   )
   EmissiveColor=Clamp(None=LinearInterpolate(A=@n90.EmissiveColor, B=@n88, Alpha=@n71)) {maxDefault=5}
  ) {attributeSetTypes [
    #2
    #3
    #4
    #5
    #1
   ]}
  n92 = FunctionInput() {bUsePreviewValueAsDefault=true, inputName="Customization Mask 2 (_CM B Top Right)", inputType=FunctionInput_Scalar, sortPriority=25}
  n93 = Add(A=@n92, B=@n72)
  n94 = NamedRerouteUsage()
  n95 = StaticSwitchParameter(True=BlendMaterialAttributes(A=@n90.MaterialAttributes, B=@n91, Alpha=LinearInterpolate(A=@n93, B=@n72, Alpha=@n94)), False=@n89) {
   desc=Top Side Projection Base - Height Lerp Anim - Reactive - Topside Emissive Lerp 2 parallax levels - Secondary color Pick
   group=Customization_Wrap
   parameterName=Wrap_ElectricBurst
  }
  FunctionOutput(None=@n95)
  NamedRerouteDeclaration(None=FunctionInput() {bUsePreviewValueAsDefault=true, inputName=isCharacter, inputType=FunctionInput_Scalar, sortPriority=4}) {name=IsCharacter, nodeColor=<rgba 0.9387 1 0 1>}
  n96 = Time()
  n97 = VectorParameter() {group=TopSideProjection_UVS, parameterName=TS_Panning_Weapon_02, sortPriority=10, defaultValue=<rgba 0 0 0 1>}
  n98 = VectorParameter() {group=TopSideProjection_UVS, parameterName=TS_Panning_Vehicle_02, sortPriority=11, defaultValue=<rgba 0 0 0 1>}
  n99 = VectorParameter() {group=TopSideProjection_UVS, parameterName=TS_Panning_Mech_02, sortPriority=12, defaultValue=<rgba 0 0 0 1>}
  n100 = LinearInterpolate(A=LinearInterpolate(A=@n97.RGB, B=@n98.RGB, Alpha=@n7), B=@n99.RGB, Alpha=@n6)
  Reroute(None=Add(A=@n11.Result, B=Multiply(A=@n96, B=@n100)))
  ScalarParameter() {desc="0 = effect is off, 1 = effect is on", group=TopSideProjection_Reactive, parameterName=TurnOnEmissive_ForSequencer}
  ScalarParameter() {group=TopSideProjection, parameterName=TS_Fill_Z_Axis_Weapon_Opacity 01, sortPriority=2}
 ]}
