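# Material function for weapon skin wraps that adds animated, triplanar normal-map "flutter" to incoming material attributes, toggled by a static switch.
# Scrolling samples use pre-skinned position and blend by contrast-adjusted normal, with a single-axis option whose plane follows a vehicle input.
# Fresnel-driven flattening fades the detail toward grazing angles before angle-corrected blending with the existing normal.
# Static switches limit the effect to base, primary or secondary customization regions, or cover the whole object.
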MaterialFunction /Game/Weapons/WeaponSkins/Wraps/MaterialFunctions/MF_FlutterNormals
graph {nodes [
  n1 = FunctionInput() {inputType=FunctionInput_MaterialAttributes}
  n2 = GetMaterialAttributes(None=@n1) {attributeGetTypes [#1]}
  n3 = ScalarParameter() {defaultValue=1, group=FlutterNormals, parameterName=FN_Normal Tiling}
  n4 = PreSkinnedPosition()
  n5 = Divide(
   A=FunctionInput(Preview=VertexInterpolator(VS=@n4).PS) {bUsePreviewValueAsDefault=true, inputName=Local Position_1}
   B=ScalarParameter() {defaultValue=40, group=FlutterNormals, parameterName=FN_Triplanner_ConstantScaleDivisor}
  )
  n6 = Frac(None=Multiply(A=Time(), B=VectorParameter() {group=FlutterNormals, parameterName=FN_TriPlanner Speed, defaultValue=<rgba 0 0.2 0.2 1>}.RGB))
  n7 = Add(A=Multiply(A=@n3, B=@n5), B=@n6)
  n8 = TextureObjectParameter() {
   group=FlutterNormals
   parameterName=FN_Normal Map
   samplerType=SAMPLERTYPE_Normal
   texture=/Engine/Functions/Engine_MaterialFunctions02/ExampleContent/Textures/water_n
  }
  n9 = TextureSample(UVs=Add(B=ComponentMask(None=@n7) {b=true, g=true}), Tex=@n8) {samplerType=SAMPLERTYPE_Normal, texture=&FlatNormalBlue}
  n10 = Add(A=Multiply(A=@n3, B=@n5), B=@n6)
  n11 = TextureSample(UVs=Add(B=ComponentMask(None=@n10) {b=true, r=true}), Tex=@n8) {samplerType=SAMPLERTYPE_Normal, texture=&FlatNormalBlue}
  n12 = FunctionInput(Preview=VertexInterpolator(VS=PreSkinnedNormal()).PS) {bUsePreviewValueAsDefault=true, inputName=Local Normal}
  n13 = Abs(None=ComponentMask(None=@n12) {g=true})
  n14 = ScalarParameter() {defaultValue=1, group=FlutterNormals, parameterName=FN_TriPlanner_AxisBlendStrength}
  n15 = Saturate(None=ComponentMask(None=MaterialFunctionCall(In=@n13, Contrast=@n14) {materialFunction=&CheapContrast}.Result) {g=true})
  n16 = Add(A=Multiply(A=@n3, B=@n5), B=@n6)
  n17 = TextureSample(UVs=Add(B=ComponentMask(None=@n16) {g=true, r=true}), Tex=@n8) {samplerType=SAMPLERTYPE_Normal, texture=&FlatNormalBlue}
  n18 = ComponentMask(None=@n12) {b=true}
  n19 = ComponentMask(None=MaterialFunctionCall(In=Abs(None=@n18), Contrast=@n14) {materialFunction=&CheapContrast}.Result) {b=true}
  n20 = LinearInterpolate(
   A=LinearInterpolate(A=LinearInterpolate(A=@n9.RGB, B=@n11.RGB, Alpha=@n15), B=@n17.RGB, Alpha=Saturate(None=@n19))
   B=LinearInterpolate(A=@n11.RGB, B=@n17.RGB, Alpha=FunctionInput() {bUsePreviewValueAsDefault=true, inputName=isVehicle_1, inputType=FunctionInput_Scalar})
   Alpha=ScalarParameter() {group=FlutterNormals, parameterName=FN_Triplanner_Z axis Only}
  )
  n21 = LinearInterpolate(
   A=ScalarParameter() {defaultValue=0.7, group=FlutterNormals, parameterName=FN_Flatness}
   Alpha=Fresnel(ExponentIn=Constant() {r=1}, Normal=VertexNormalWS()) {baseReflectFraction=0}
  )
  n22 = SetMaterialAttributes!(
   MaterialAttributes=@n2.MaterialAttributes
   Normal=MaterialFunctionCall(
    BaseNormal=@n2.Normal
    AdditionalNormal=MaterialFunctionCall(Normal=@n20, Flatness=@n21) {materialFunction=&FlattenNormal}.Result
   ) {materialFunction=&BlendAngleCorrectedNormals}.Result
  ) {attributeSetTypes [#1]}
  n23 = FunctionInput() {bUsePreviewValueAsDefault=true, inputName="Customization Mask 2 (_CM B Top Right)", inputType=FunctionInput_Scalar, sortPriority=25}
  n24 = FunctionInput() {bUsePreviewValueAsDefault=true, inputName="Customization Mask (_CM B Top left)", inputType=FunctionInput_Scalar, sortPriority=24}
  n25 = Add(A=@n23, B=@n24) {desc="Base + Secondary = exposed primary"}
  n26 = OneMinus(None=@n25) {desc=Primary}
  n27 = Add(A=@n26, B=@n24)
  n28 = StaticSwitchParameter(
   True=@n25
   False=StaticSwitchParameter(
    True=@n26
    False=StaticSwitchParameter(True=@n23, False=@n24) {desc=False uses Base customization, group=FlutterNormals, parameterName=FN_OnlySecondary, sortPriority=-99}
   ) {group=FlutterNormals, parameterName=FN_OnlyPrimary, sortPriority=-98}
  ) {group=FlutterNormals, parameterName=FN_BaseAndSecondary, sortPriority=-97}
  n29 = StaticSwitchParameter(
   True=Constant() {r=1}
   False=StaticSwitchParameter(
    True=Add(A=@n26, B=@n23)
    False=StaticSwitchParameter(True=@n27, False=@n28) {group=FlutterNormals, parameterName=FN_BaseAndPrimary, sortPriority=-96}
   ) {group=FlutterNormals, parameterName=FN_SecondaryAndPrimary, sortPriority=-95}
  ) {defaultValue=true, group=FlutterNormals, parameterName=FN_CoverEntireObject, sortPriority=-100}
  FunctionOutput(None=StaticSwitchParameter(True=BlendMaterialAttributes(A=@n1, B=@n22, Alpha=@n29), False=@n1) {desc=Triplanar Projection - adds uniform normal noise, group=Customization, parameterName=Effect_Flutter_Normals})
 ]}
