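# Material function for wraps that blends a flat-colored SDF pattern layer onto the input material, toggled by a static switch.
# Pre-skinned bounds and normals drive side-projected UVs, with offset, scale and rotation chosen by object type (weapon, vehicle, mech, character, backpack, pickaxe).
# SDF mask channels are thresholded and morphed manually, by sine, or reactively from a timestamp parameter that can also drive rotation, color and glow.
# Input masks and a normal-based projection mask limit the blend.
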MaterialFunction /Game/Weapons/WeaponSkins/Wraps/MaterialFunctions/MF_Function_SDF_Blend
graph {nodes [
  n1 = NamedRerouteUsage()
  n2 = NamedRerouteUsage()
  n3 = TextureSample() {samplerType=SAMPLERTYPE_Masks, texture=/Game/Weapons/FORT_Rifles/Textures/T_Scar_CM}
  n4 = FunctionInput(Preview=@n3.R) {bUsePreviewValueAsDefault=true, inputName=Mask Primary, inputType=FunctionInput_Scalar, sortPriority=24, previewValue=<1 0 0 1>}
  n5 = FunctionInput(Preview=@n3.B) {bUsePreviewValueAsDefault=true, inputName=Mask Secondary, inputType=FunctionInput_Scalar, sortPriority=24, previewValue=<1 0 0 1>}
  n6 = Multiply(A=@n5, B=ScalarParameter() {defaultValue=1, group=Function_SDF_Blend, parameterName="SDFBL_Primary only 0=ON", sortPriority=0})
  n7 = Multiply(A=Saturate(None=Add(A=@n4, B=@n6)), B=Multiply(A=NamedRerouteUsage(), B=NamedRerouteUsage()))
  n8 = TextureSample() {samplerType=SAMPLERTYPE_Normal, texture=/Game/Weapons/FORT_Rifles/T_SCAR_N}
  n9 = Constant3Vector() {constant=<rgba 1 0 0 0>}
  n10 = GetMaterialAttributes(None=FunctionInput(Preview=SetMaterialAttributes!(MaterialAttributes, Normal=@n8.RGB, BaseColor=@n9) {attributeSetTypes [#1, #2]}) {inputName=Material, inputType=FunctionInput_MaterialAttributes}) {attributeGetTypes [
    #2
    #3
    #1
    #4
   ]}
  FunctionOutput(None=StaticSwitchParameter(
    True=BlendMaterialAttributes(A=@n1, B=@n2, Alpha=@n7) {vertexAttributeBlendType=UseA}
    False=NamedRerouteDeclaration(None=@n10.MaterialAttributes) {name=Root_Material, nodeColor=<rgba 0.7812 1 0 1>}
   ) {group=Customization, parameterName=Function_SDF_Blend})
  n11 = NamedRerouteUsage()
  n12 = Constant4Vector() {constant=<rgba 0 0 0 0.5>}
  n13 = FunctionInput(Preview=@n12.RGBA) {bUsePreviewValueAsDefault=true, inputName=Object-SafeZone-Back, inputType=FunctionInput_Vector4, sortPriority=100}
  n14 = Multiply(
   A=@n11
   B=FunctionInput(Preview=@n12.RGBA) {bUsePreviewValueAsDefault=true, inputName=Object-SafeZone-Middle, inputType=FunctionInput_Vector4, sortPriority=101}
  )
  n15 = StaticSwitchParameter(
   True=Multiply(
    A=@n11
    B=FunctionInput(Preview=@n12.RGBA) {bUsePreviewValueAsDefault=true, inputName=Object-SafeZone-Front, inputType=FunctionInput_Vector4, sortPriority=102}
   )
   False=NamedRerouteUsage()
  ) {group=Function_SDF_Blend, parameterName=SDFBL_UV_SafeZone_Front}
  NamedRerouteDeclaration(None=StaticSwitchParameter(
    True=Multiply(A=@n11, B=@n13)
    False=StaticSwitchParameter(True=@n14, False=@n15) {group=Function_SDF_Blend, parameterName=SDFBL_UV_SafeZone_Middle}
   ) {group=Function_SDF_Blend, parameterName=SDFBL_UV_SafeZone_Back}) {name=Scale/offse, nodeColor=<rgba 0 0.9297 1 1>}
  n16 = Subtract(A=Time(), B=ScalarParameter() {defaultValue=-9999999, group=Function_SDF_Blend_Reactive, parameterName=SDFBL_Time_BP})
  n17 = Saturate(None=Divide(A=@n16, B=ScalarParameter() {defaultValue=0.25, group=Function_SDF_Blend_Reactive, parameterName=SDFBL_Glow_Duration}))
  n18 = ScalarParameter() {defaultValue=0.5, group=Function_SDF_Blend_Reactive, parameterName=SDFBL_Glow_Curve}
  NamedRerouteDeclaration(None=LinearInterpolate(
    A=ScalarParameter() {defaultValue=1, group=Function_SDF_Blend_Reactive, parameterName=SDFBL_Glow_Start, sortPriority=10}
    B=ScalarParameter() {group=Function_SDF_Blend_Reactive, parameterName=SDFBL_Glow_End, sortPriority=10}
    Alpha=Power(Base=@n17, Exp=@n18)
   )) {name=TimeBP_Glow, nodeColor=<rgba 1 0.2812 0 1>}
  n19 = MaterialFunctionCall(Float4=NamedRerouteUsage()) {materialFunction=&BreakOutFloat4Components}
  n20 = NamedRerouteUsage()
  n21 = Saturate(None=MaterialFunctionCall!(
    For_Safe_Zones=StaticBool() {value=true}
    Normal_Map
    Caculated_Bounds=NamedRerouteUsage()
    Pre_Skinned_Normals=NamedRerouteUsage()
    Is Vehicle=@n19.R
    Is VehicleSide=@n19.R
    Is Mech=@n19.R
    Vehicle_Rotation
    UV_Rotation Bypass=NamedRerouteUsage()
    Mech_Rotation
    Uvs_Offset=ComponentMask(None=@n20) {b=true, g=true, r=true}
    Offset-after rotation
    Uvs_TileRate=ComponentMask(None=@n20) {a=true}
   ) {materialFunction=&/MF_Projection_Side_Normal}.UVs)
  n22 = TextureSampleParameter2D(UVs=@n21) {
   group=Function_SDF_Blend
   parameterName="SDFBL__Tex_SDF_Msk 1,2"
   samplerType=SAMPLERTYPE_LinearColor
   sortPriority=10
   texture=/Game/Weapons/WeaponSkins/Wraps/Textures/WildCard/T_Wrap_WildCard_Msk
  }
  n23 = ScalarParameter() {defaultValue=0.49, group=Function_SDF_Blend, parameterName=SDFBL_SDF_Min}
  n24 = ScalarParameter() {defaultValue=0.5, group=Function_SDF_Blend, parameterName=SDFBL_SDF_Max}
  n25 = Divide(A=Subtract(A=@n22.RGBA, B=@n23), B=Subtract(A=@n24, B=@n23))
  n26 = ComponentMask(None=@n25) {r=true}
  n27 = ComponentMask(None=@n25) {a=true}
  n28 = TextureSampleParameter2D(UVs=@n21) {
   group=Function_SDF_Blend
   parameterName="SDFBL__Tex_SDF_Msk_3,4"
   samplerType=SAMPLERTYPE_LinearColor
   sortPriority=10
   texture=/Game/Weapons/WeaponSkins/Wraps/Textures/WildCard/T_Wrap_WildCard_Msk_1
  }
  n29 = Divide(A=Subtract(A=@n28.RGBA, B=@n23), B=Subtract(A=@n24, B=@n23))
  n30 = AppendVector(A=AppendVector(A=@n26, B=@n27), B=AppendVector(A=ComponentMask(None=@n29) {r=true}, B=ComponentMask(None=@n29) {a=true}))
  n31 = VectorParameter() {group=Function_SDF_Blend_Channels, parameterName="SDFBL_Channel_Mask_1,2,3,_B", defaultValue=<rgba 0 0 1 1>}
  n32 = ScalarParameter() {group=Function_SDF_Blend_Channels, parameterName=SDFBL_Channel_Mask_4_B}
  n33 = VectorParameter() {group=Function_SDF_Blend_Channels, parameterName="SDFBL_Channel_Mask_1,2,3,_A", defaultValue=<rgba 1 0 0 1>}
  n34 = ScalarParameter() {group=Function_SDF_Blend_Channels, parameterName=SDFBL_Channel_Mask_4_A}
  n35 = ScalarParameter() {group=Function_SDF_Blend, parameterName=SDFBL_SDF_Morph_Speed}
  n36 = Sine(None=Multiply(A=Time(), B=@n35))
  n37 = Multiply(A=Add(A=@n36)) {constB=0.5}
  n38 = Ceil(None=@n35)
  n39 = StaticSwitchParameter(
   True=NamedRerouteUsage()
   False=Add(A=ScalarParameter() {group=Function_SDF_Blend, parameterName=SDFBL_SDF_Morph}, B=LinearInterpolate(B=@n37, Alpha=@n38))
  ) {group=Function_SDF_Blend_Reactive, parameterName=SDFBL_Reactive, sortPriority=0}
  n40 = LinearInterpolate(A=DotProduct(A=@n30, B=AppendVector(A=@n31.RGB, B=@n32)), B=DotProduct(A=@n30, B=AppendVector(A=@n33.RGB, B=@n34)), Alpha=@n39)
  n41 = LinearInterpolate(A=@n26, B=@n27, Alpha=@n39)
  NamedRerouteDeclaration(None=Saturate(None=StaticSwitchParameter(True=@n40, False=@n41) {group=Function_SDF_Blend, parameterName="SDFBL_2---->4", sortPriority=0})) {name=Mask, nodeColor=<rgba 0 0.5312 1 1>}
  n42 = Saturate(None=Divide(A=@n16, B=ScalarParameter() {defaultValue=0.25, group=Function_SDF_Blend_Reactive, parameterName=SDFBL_Duration}))
  NamedRerouteDeclaration(None=LinearInterpolate(
    A=ScalarParameter() {defaultValue=1, group=Function_SDF_Blend_Reactive, parameterName=SDFBL_Start, sortPriority=5}
    B=ScalarParameter() {group=Function_SDF_Blend_Reactive, parameterName=SDFBL_End, sortPriority=5}
    Alpha=@n42
   )) {name=TimeBP, nodeColor=<rgba 1 0.2812 0 1>}
  NamedRerouteDeclaration(None=LinearInterpolate(
    A=ScalarParameter() {defaultValue=1, group=Function_SDF_Blend_Reactive, parameterName=SDFBL_Rotation_Start, sortPriority=10}
    B=ScalarParameter() {group=Function_SDF_Blend_Reactive, parameterName=SDFBL_Rotation_End, sortPriority=10}
    Alpha=@n42
   )) {name=TimeBP_Rotation, nodeColor=<rgba 1 0.2812 0 1>}
  NamedRerouteDeclaration(None=Multiply(A=@n36, B=ScalarParameter() {group=Function_SDF_Blend, parameterName=SDFBL_SDF_Rotation_Amount})) {name=Time_Rotation, nodeColor=<rgba 0.1484 1 0 1>}
  n43 = ScalarParameter() {group=Function_SDF_Blend_UVs, parameterName=SDFBL_Rotation_Weapon, sortPriority=150}
  n44 = MaterialFunctionCall(Float4=NamedRerouteUsage()) {materialFunction=&BreakOutFloat4Components}
  n45 = LinearInterpolate(
   A=ScalarParameter() {group=Function_SDF_Blend_UVs, parameterName=SDFBL_Rotation_Vehicle, sortPriority=150}
   B=ScalarParameter() {group=Function_SDF_Blend_UVs, parameterName=SDFBL_Rotation_Vehicle_Side, sortPriority=150}
   Alpha=@n44.G
  )
  n46 = ScalarParameter() {group=Function_SDF_Blend_UVs, parameterName=SDFBL_Rotation_Mech, sortPriority=150}
  n47 = ScalarParameter() {group=Function_SDF_Blend_UVs, parameterName=SDFBL_Rotation_Character, sortPriority=150}
  n48 = MaterialFunctionCall(Float2=NamedRerouteUsage()) {materialFunction=&BreakOutFloat2Components}
  n49 = LinearInterpolate(
   A=LinearInterpolate(A=LinearInterpolate(A=@n43, B=@n45, Alpha=@n44.R), B=LinearInterpolate(A=@n46, B=@n47, Alpha=@n44.A), Alpha=@n44.B)
   B=ScalarParameter() {group=Function_SDF_Blend_UVs, parameterName=SDFBL_Rotation_Backpack, sortPriority=150}
   Alpha=@n48.R
  )
  n50 = ScalarParameter() {group=Function_SDF_Blend_UVs, parameterName=SDFBL_Rotation_Pickaxe, sortPriority=150}
  n51 = NamedRerouteUsage()
  n52 = NamedRerouteUsage()
  NamedRerouteDeclaration(None=Add(
    A=LinearInterpolate(A=@n49, B=@n50, Alpha=@n48.G)
    B=StaticSwitchParameter(True=@n51, False=@n52) {group=Function_SDF_Blend_Reactive, parameterName=SDFBL_Reactive, sortPriority=0}
   )) {name=UVs_Rotation, nodeColor=<rgba 0 0.7188 1 1>}
  n53 = VectorParameter() {group=Function_SDF_Blend, parameterName=SDFBL_SDF_BCB, sortPriority=5, defaultValue=<rgba 0 0.0029 1 1>}
  n54 = VectorParameter() {group=Function_SDF_Blend, parameterName=SDFBL_SDF_BC, sortPriority=5, defaultValue=<rgba 0 1 0.0309 1>}
  n55 = NamedRerouteUsage()
  n56 = VectorParameter() {group=Function_SDF_Blend, parameterName=SDFBL_SDF_Glow, sortPriority=5, defaultValue=<rgba 0 0 0 1>}
  n57 = Multiply(A=@n56.RGB, B=NamedRerouteUsage())
  n58 = Constant()
  n59 = SetMaterialAttributes!(
   MaterialAttributes=NamedRerouteUsage()
   BaseColor=StaticSwitchParameter(True=LinearInterpolate(A=@n53.RGB, B=@n54.RGB, Alpha=@n55), False=@n54.RGB) {group=Function_SDF_Blend, parameterName=SDFBL_Reactive_Color, sortPriority=0}
   Roughness=ScalarParameter() {group=Function_SDF_Blend, parameterName=SDFBL_SDF_Roughness, sortPriority=5}
   Specular=ScalarParameter() {defaultValue=0.5, group=Function_SDF_Blend, parameterName=SDFBL_SDF_Spec, sortPriority=5}
   Metallic=ScalarParameter() {group=Function_SDF_Blend, parameterName=SDFBL_SDF_Metalic, sortPriority=5}
   Normal
   EmissiveColor=Add(
    A=NamedRerouteUsage()
    B=StaticSwitchParameter(True=@n57, False=@n58) {group=Function_SDF_Blend_Reactive, parameterName=SDFBL_Reactive, sortPriority=0}
   )
  ) {attributeSetTypes [
    #2
    #3
    #5
    #4
    #1
    #6
   ]}
  NamedRerouteDeclaration(None=@n59) {name=Mat_Base, nodeColor=<rgba 0 0.3906 1 1>}
  n60 = VectorParameter() {group=Function_SDF_Blend_UVs, parameterName=SDFBL_Offset_Scale_Weapon, defaultValue=<rgba 0 0 0 0.5>}
  n61 = VectorParameter() {group=Function_SDF_Blend_UVs, parameterName=SDFBL_Offset_Scale_Vehicle, defaultValue=<rgba 0 0 0 1>}
  n62 = AppendVector(A=@n61.RGB, B=@n61.RGB)
  n63 = VectorParameter() {group=Function_SDF_Blend_UVs, parameterName=SDFBL_Offset_Scale_VehicleSide, defaultValue=<rgba 0 0 0 1>}
  n64 = AppendVector(A=@n63.RGB, B=@n63.RGB)
  n65 = MaterialFunctionCall(Float4=NamedRerouteUsage()) {materialFunction=&BreakOutFloat4Components}
  n66 = VectorParameter() {group=Function_SDF_Blend_UVs, parameterName=SDFBL_Offset_Scale_Mech, defaultValue=<rgba 0 0 0 1>}
  n67 = VectorParameter() {group=Function_SDF_Blend_UVs, parameterName=SDFBL_Offset_Scale_Character, defaultValue=<rgba 0 0 0 1>}
  n68 = LinearInterpolate(
   A=LinearInterpolate(A=AppendVector(A=@n60.RGB, B=@n60.RGB), B=LinearInterpolate(A=@n62, B=@n64, Alpha=@n65.G), Alpha=@n65.R)
   B=LinearInterpolate(A=AppendVector(A=@n66.RGB, B=@n66.RGB), B=AppendVector(A=@n67.RGB, B=@n67.RGB), Alpha=@n65.A)
   Alpha=@n65.B
  )
  n69 = VectorParameter() {group=Function_SDF_Blend_UVs, parameterName=SDFBL_Offset_Scale_Backpack, defaultValue=<rgba 0 0 0 1>}
  n70 = AppendVector(A=@n69.RGB, B=@n69.RGB)
  n71 = MaterialFunctionCall(Float2=NamedRerouteUsage()) {materialFunction=&BreakOutFloat2Components}
  n72 = VectorParameter() {group=Function_SDF_Blend_UVs, parameterName=SDFBL_Offset_Scale_Pickaxe, defaultValue=<rgba 0 0 0 1>}
  NamedRerouteDeclaration(None=LinearInterpolate(A=LinearInterpolate(A=@n68, B=@n70, Alpha=@n71.R), B=AppendVector(A=@n72.RGB, B=@n72.RGB), Alpha=@n71.G)) {name=UVs_Offset_Scale, nodeColor=<rgba 0 0.625 1 1>}
  n73 = ScalarParameter() {defaultValue=1, group=Function_SDF_Blend_UVs, parameterName=SDFBL_Scale_Weapon, sortPriority=150}
  n74 = MaterialFunctionCall(Float4=NamedRerouteUsage()) {materialFunction=&BreakOutFloat4Components}
  n75 = LinearInterpolate(
   A=ScalarParameter() {defaultValue=1, group=Function_SDF_Blend_UVs, parameterName=SDFBL_Scale_Vehicle, sortPriority=150}
   B=ScalarParameter() {defaultValue=1, group=Function_SDF_Blend_UVs, parameterName=SDFBL_Scale_Vehicle_Side, sortPriority=150}
   Alpha=@n74.G
  )
  n76 = ScalarParameter() {defaultValue=1, group=Function_SDF_Blend_UVs, parameterName=SDFBL_Scale_Mech, sortPriority=150}
  n77 = ScalarParameter() {defaultValue=1, group=Function_SDF_Blend_UVs, parameterName=SDFBL_Scale_Character, sortPriority=150}
  n78 = MaterialFunctionCall(Float2=NamedRerouteUsage()) {materialFunction=&BreakOutFloat2Components}
  n79 = LinearInterpolate(
   A=LinearInterpolate(A=LinearInterpolate(A=@n73, B=@n75, Alpha=@n74.R), B=LinearInterpolate(A=@n76, B=@n77, Alpha=@n74.A), Alpha=@n74.B)
   B=ScalarParameter() {defaultValue=1, group=Function_SDF_Blend_UVs, parameterName=SDFBL_Scale_Backpack, sortPriority=150}
   Alpha=@n78.R
  )
  n80 = ScalarParameter() {defaultValue=1, group=Function_SDF_Blend_UVs, parameterName=SDFBL_Scale_Pickaxe, sortPriority=150}
  NamedRerouteDeclaration(None=AppendVector(A=Constant3Vector() {constant=<rgba 1 1 1 0>}, B=LinearInterpolate(A=@n79, B=@n80, Alpha=@n78.G))) {name=UVs_Scale, nodeColor=<rgba 0 0.7188 1 1>}
  n81 = MaterialFunctionCall(Color=Abs(None=NamedRerouteUsage())) {materialFunction=&SplitComponents}
  n82 = MaterialFunctionCall(Float4=NamedRerouteUsage()) {materialFunction=&BreakOutFloat4Components}
  n83 = LinearInterpolate(A=@n81.B, B=@n81.B, Alpha=@n82.G)
  n84 = MaterialFunctionCall(
   In=LinearInterpolate(A=LinearInterpolate(A=@n81.R, B=@n83, Alpha=@n82.R), B=@n81.G, Alpha=@n82.B)
   Contrast=ScalarParameter() {defaultValue=1, group=Function_SDF_Blend, parameterName=SDFBL_Blend_Contrast, sortPriority=0}
  ) {materialFunction=&CheapContrast}
  n85 = LinearInterpolate(
   A=ScalarParameter() {defaultValue=1, group=Function_SDF_Blend, parameterName="SDFBL_Blend_Tri  0=ON", sortPriority=0}
   B=ScalarParameter() {defaultValue=1, group=Function_SDF_Blend, parameterName="SDFBL_Blend_Tri_VehicleSide  0=ON", sortPriority=0}
   Alpha=@n82.G
  )
  NamedRerouteDeclaration(None=Saturate(None=Add(A=@n84.Result, B=@n85) {constB=0})) {name=Tri_Mask, nodeColor=<rgba 1 0.4688 0 1>}
  n86 = GetMaterialAttributes(None=NamedRerouteUsage()) {attributeGetTypes [
    #2
    #4
    #5
    #3
    #6
   ]}
  NamedRerouteDeclaration(None=@n86.EmissiveColor) {name=Root_Emissive, nodeColor=<rgba 0 0.7422 1 1>}
  NamedRerouteDeclaration(None=FunctionInput(Preview=Constant4Vector().RGBA) {bUsePreviewValueAsDefault=true, inputName=IsState, inputType=FunctionInput_Vector4, sortPriority=4}) {name=IsState, nodeColor=<rgba 0.1497 1 0.1633 1>}
  NamedRerouteDeclaration(None=FunctionInput(Preview=Constant2Vector()) {bUsePreviewValueAsDefault=true, inputName=IsState_B, inputType=FunctionInput_Vector2, sortPriority=4}) {name=IsState_B, nodeColor=<rgba 0.1497 1 0.1633 1>}
  n87 = PreSkinnedLocalBounds()
  n88 = Subtract(A=VertexInterpolator(VS=PreSkinnedPosition()).PS, B=@n87.Min)
  n89 = MaterialFunctionCall(Color=Subtract(A=@n87.Max, B=@n87.Max)) {materialFunction=&SplitComponents}
  n90 = Max(A=Max(A=@n89.R, B=@n89.R), B=@n89.B)
  NamedRerouteDeclaration(None=FunctionInput(Preview=Divide(A=@n88, B=@n90)) {bUsePreviewValueAsDefault=true, inputName=Calcated_Bounds, sortPriority=11}) {name=Caclcated_Bounds, nodeColor=<rgba 0 0.9297 1 1>}
  n91 = PreSkinnedNormal()
  NamedRerouteDeclaration(None=FunctionInput(Preview=VertexInterpolator(VS=@n91).PS) {bUsePreviewValueAsDefault=true, inputName=Pre-Skinned Local Normal, sortPriority=10}) {name=PreSkin_Normal, nodeColor=<rgba 0.8281 1 0 1>}
  NamedRerouteDeclaration(None=@n10.Normal) {name=Root_Normal, nodeColor=<rgba 0.4062 0 1 1>}
 ]}
