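# Wrap material function that, behind the UseGlitchyFx switch, adds a glowing, glitching EvilAstronaut logo and animated glitch cubes to a material's emissive.
# Bounds-based local position becomes scaled logo UVs, distorted by MF_GlitchFunction and edge-masked.
# Time-seeded hashes of quantized pre-skinned position form cubes at two resolutions, selected by a third hash.
# Weapon, vehicle and mech inputs blend projections and scales; a second logo branch is unconnected to the output.
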MaterialFunction /Game/Weapons/WeaponSkins/Wraps/MaterialFunctions/MF_GlitchyFX
graph {nodes [
  n1 = FunctionInput(Preview=MakeMaterialAttributes!(
    BaseColor
    Metallic
    Specular
    Roughness
    Anisotropy
    EmissiveColor
    Opacity
    OpacityMask
    Normal
    Tangent
    WorldPositionOffset
    SubsurfaceColor
    ClearCoat
    ClearCoatRoughness
    AmbientOcclusion
    Refraction
    CustomizedUVs_0
    CustomizedUVs_1
    CustomizedUVs_2
    CustomizedUVs_3
    CustomizedUVs_4
    CustomizedUVs_5
    CustomizedUVs_6
    CustomizedUVs_7
    PixelDepthOffset
    ShadingModel
    Displacement
   )) {bUsePreviewValueAsDefault=true, inputName=Material, inputType=FunctionInput_MaterialAttributes}
  n2 = GetMaterialAttributes(None=@n1) {attributeGetTypes [#1]}
  n3 = PreSkinnedLocalBounds()
  n4 = Subtract(A=VertexInterpolator(VS=PreSkinnedPosition()).PS, B=@n3.Min)
  n5 = Subtract(A=@n3.Max, B=@n3.Max)
  n6 = Multiply(
   A=FunctionInput(Preview=Divide(A=@n4, B=@n5)) {bUsePreviewValueAsDefault=true, inputName=Bounds}
   B=Constant3Vector() {constant=<rgba 1 1 -1 0>}
  )
  n7 = Constant3Vector() {constant=<rgba 0 0 1 0>}
  n8 = VectorParameter() {group=GlitchFX, parameterName=SkullMech_Offset}
  n9 = FunctionInput() {bUsePreviewValueAsDefault=true, inputName=IsMech, inputType=FunctionInput_Scalar}
  n10 = MaterialFunctionCall(Color=Add(A=Add(A=@n6, B=@n7), B=LinearInterpolate(B=@n8.RGB, Alpha=@n9))) {materialFunction=&SplitComponents}
  n11 = FunctionInput() {bUsePreviewValueAsDefault=true, inputName=Side Project Vehicle, inputType=FunctionInput_Scalar}
  n12 = FunctionInput() {bUsePreviewValueAsDefault=true, inputName=IsVehicle, inputType=FunctionInput_Scalar}
  n13 = LinearInterpolate(
   A=AppendVector(A=@n10.G, B=@n10.G)
   B=LinearInterpolate(A=AppendVector(A=@n10.G, B=@n10.G), B=AppendVector(A=@n10.R, B=@n10.R), Alpha=@n11)
   Alpha=@n12
  )
  n14 = AppendVector(A=@n10.R, B=@n10.R)
  n15 = LinearInterpolate(
   A=ScalarParameter() {defaultValue=0.25, group=GlitchFX, parameterName=SkullWeapon_Scale}
   B=LinearInterpolate(
    A=ScalarParameter() {defaultValue=0.25, group=GlitchFX, parameterName=SkullVehicle_Scale}
    B=ScalarParameter() {defaultValue=0.25, group=GlitchFX, parameterName=SkullVehicleSide_Scale}
    Alpha=@n11
   )
   Alpha=@n12
  )
  n16 = ScalarParameter() {defaultValue=0.25, group=GlitchFX, parameterName=SkulMech_Scale}
  n17 = MaterialFunctionCall!(UV=MaterialFunctionCall(UVs=LinearInterpolate(A=@n13, B=@n14, Alpha=@n9), Texture Scale=LinearInterpolate(A=@n15, B=@n16, Alpha=@n9)) {materialFunction=&ScaleUVsByCenter}.UVs) {materialFunction=/Game/Weapons/WeaponSkins/MaterialFunction/MF_GlitchFunction}
  n18 = TextureObject() {texture=/Game/Weapons/WeaponSkins/Wraps/Textures/EvilAstronaut/T_EvilAstronaut_Logo_Wrap_02}
  n19 = TextureSample(UVs=@n17.Glitched UV, Tex=@n18) {texture=&T_EvilAstronaut_Logo_Wrap_02}
  n20 = Constant() {r=0.5}
  n21 = MaterialFunctionCall(Float2=@n17.Glitched UV) {materialFunction=&BreakOutFloat2Components}
  n22 = Saturate(None=Multiply(A=SphereMask(A=@n20, B=@n21.R) {attenuationRadius=0.5}, B=SphereMask(A=@n21.G, B=@n20) {attenuationRadius=0.5}))
  n23 = Multiply(
   A=ChannelMaskParameter(None=Multiply(A=@n19.RGB, B=@n22)) {parameterName=LogoChannel}.R
   B=VectorParameter() {group=GlitchFX, parameterName=LogoGlowColor, defaultValue=<rgba 1 0 0 1>}.RGB
  )
  n24 = ScalarParameter() {defaultValue=5, group=GlitchFX, parameterName=LogoGlowStrength}
  n25 = Time()
  n26 = Multiply(A=@n25, B=ScalarParameter() {defaultValue=0.1, group=GlitchFX, parameterName=GlitchcubeSquareSpeed})
  n27 = FunctionInput(Preview=VertexInterpolator(VS=PreSkinnedPosition()).PS) {bUsePreviewValueAsDefault=true, inputName=PreskinnedLocalPosition}
  n28 = ScalarParameter() {defaultValue=2, group=GlitchFX, parameterName=GlitchcubeRandomColorMaxResolution}
  n29 = ScalarParameter() {defaultValue=2, group=GlitchFX, parameterName=VehicleGlitchcubeRandomColorMaxResolution}
  n30 = Divide(
   A=@n27
   B=LinearInterpolate(
    A=LinearInterpolate(A=@n28, B=@n29, Alpha=@n12)
    B=ScalarParameter() {defaultValue=2, group=GlitchFX, parameterName=MechGlitchcubeRandomColorMaxResolution}
    Alpha=@n9
   )
  )
  n31 = MaterialFunctionCall(Seed=@n26, Vec3=Floor(None=Add(A=@n30) {constB=0.5})) {materialFunction=&MF_Hash}
  n32 = ScalarParameter() {defaultValue=0.7, group=GlitchFX, parameterName=GlitchCubeVarianceFalloff}
  n33 = Subtract(A=ComponentMask(None=ConstantBiasScale(None=@n31.Vec3)) {r=true}, B=@n32) {constB=0.8}
  n34 = ScalarParameter() {defaultValue=7.8483, group=GlitchFX, parameterName=GlitchcubeRandomColorMinResolution}
  n35 = ScalarParameter() {defaultValue=7.8483, group=GlitchFX, parameterName=VehicleGlitchcubeRandomColorMinResolution}
  n36 = Divide(
   A=@n27
   B=LinearInterpolate(
    A=LinearInterpolate(A=@n34, B=@n35, Alpha=@n12)
    B=ScalarParameter() {defaultValue=7.8483, group=GlitchFX, parameterName=MechGlitchcubeRandomColorMinResolution}
    Alpha=@n9
   )
  )
  n37 = MaterialFunctionCall(Seed=@n26, Vec3=Floor(None=Add(A=@n36) {constB=0.5})) {materialFunction=&MF_Hash}
  n38 = Subtract(A=ComponentMask(None=ConstantBiasScale(None=@n37.Vec3)) {r=true}, B=@n32) {constB=0.8}
  n39 = ScalarParameter() {defaultValue=2, group=GlitchFX, parameterName=GlitchcubePixellationSpeed}
  n40 = LinearInterpolate(
   A=ScalarParameter() {defaultValue=2, group=GlitchFX, parameterName=GlitchcubeMaskResolution}
   B=ScalarParameter() {defaultValue=2, group=GlitchFX, parameterName=VehicleGlitchcubeMaskResolution}
   Alpha=@n12
  )
  n41 = ScalarParameter() {defaultValue=2, group=GlitchFX, parameterName=MechGlitchcubeMaskResolution}
  n42 = Add(A=Divide(A=@n27, B=LinearInterpolate(A=@n40, B=@n41, Alpha=@n9)))
  n43 = ConstantBiasScale(None=MaterialFunctionCall(Seed=Multiply(A=@n25, B=@n39), Vec3=Floor(None=@n42)) {materialFunction=&MF_Hash}.Vec3)
  n44 = LinearInterpolate(
   A=Multiply(A=Ceil(None=@n33), B=Clamp(None=@n33))
   B=Multiply(A=Ceil(None=@n38), B=Clamp(None=@n38))
   Alpha=Round(None=ComponentMask(None=@n43) {g=true})
  )
  n45 = Multiply(
   A=Clamp(None=Multiply(A=@n44) {constB=5})
   B=VectorParameter() {group=GlitchFX, parameterName=Glitch Glow Color, defaultValue=<rgba 1 0 0 1>}.RGB
  )
  n46 = ScalarParameter() {defaultValue=10, group=GlitchFX, parameterName=GlowStrength}
  n47 = Add(A=@n2.EmissiveColor, B=Add(A=Multiply(A=@n23, B=@n24), B=Multiply(A=@n45, B=@n46)))
  FunctionOutput(None=StaticSwitchParameter(
    True=SetMaterialAttributes!(MaterialAttributes=@n2.MaterialAttributes, EmissiveColor=@n47) {attributeSetTypes [#1]}
    False=@n1
   ) {group=Customization_Wrap, parameterName=UseGlitchyFx})
  n48 = VectorParameter() {group=GlitchFX, parameterName=WaaponOffset, defaultValue=<rgba 0 0.1 0 1>}
  n49 = Divide(A=@n27, B=ScalarParameter() {defaultValue=48, group=GlitchFX, parameterName=SkullWeaponScale})
  n50 = Constant2Vector() {g=0.5, r=0.5}
  n51 = MaterialFunctionCall(
   UVs=ComponentMask(None=Add(A=@n48.RGB, B=@n49)) {b=true, g=true}
   Rotation Center=@n50
   "Rotation Angle (0-1)"=ScalarParameter() {defaultValue=0.5, group=GlitchFX, parameterName=WeaponRotateLogo}
  ) {materialFunction=&CustomRotator}
  n52 = Divide(A=@n27, B=ScalarParameter() {defaultValue=48, group=GlitchFX, parameterName=SkullVehicleScale})
  n53 = VectorParameter() {group=GlitchFX, parameterName=VehicleOffset, defaultValue=<rgba 0 0.1 0 1>}
  n54 = MaterialFunctionCall(
   UVs=ComponentMask(None=Add(A=@n52, B=@n53.RGB)) {g=true, r=true}
   Rotation Center=@n50
   "Rotation Angle (0-1)"=ScalarParameter() {defaultValue=0.5, group=GlitchFX, parameterName=VehicleRotateLogo}
  ) {materialFunction=&CustomRotator}
  n55 = ScalarParameter() {defaultValue=48, group=GlitchFX, parameterName=SkullMechScale}
  n56 = ComponentMask(None=Add(A=Divide(A=@n27, B=@n55), B=VectorParameter() {group=GlitchFX, parameterName=MechOffset, defaultValue=<rgba 0 0.1 0 1>}.RGB)) {b=true, r=true}
  n57 = ScalarParameter() {defaultValue=0.5, group=GlitchFX, parameterName=MechRotateLogo}
  n58 = MaterialFunctionCall(UV=LinearInterpolate(
    A=LinearInterpolate(A=@n51.Rotated Values, B=@n54.Rotated Values, Alpha=@n12)
    B=MaterialFunctionCall(UVs=@n56, Rotation Center=@n50, "Rotation Angle (0-1)"=@n57) {materialFunction=&CustomRotator}.Rotated Values
    Alpha=@n9
   )) {materialFunction=&MF_GlitchFunction}
  n59 = Abs(None=Subtract(A=@n58.Glitched UV, B=Constant2Vector() {g=0.5, r=0.5}))
  n60 = Ceil(None=Desaturation(None=Round(None=@n59), Fraction=Constant() {r=1}))
  StaticSwitchParameter(True=Multiply(
    A=TextureSample(UVs=@n58.Glitched UV, Tex=@n18) {texture=&T_EvilAstronaut_Logo_Wrap_02}.RGB
    B=OneMinus(None=@n60)
   )) {parameterName="Glitch New-->old"}
 ]}
