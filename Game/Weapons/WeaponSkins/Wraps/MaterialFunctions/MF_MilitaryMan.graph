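# Material function for the Military Man weapon wrap, blending its look over incoming material attributes by Customization Mask, behind the UseMilitaryMan static switch.
# Local position and normal drive triplanar projection of diffuse and normal textures; red gives greyscale base color, alpha drives roughness.
# Cubed local Z over bounds applies Sun Dried Wear tint; the flattened normal blends angle-corrected over the original.
# The isVehicle Z-to-X axis lerp feeds nothing.
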MaterialFunction /Game/Weapons/WeaponSkins/Wraps/MaterialFunctions/MF_MilitaryMan
graph {nodes [
  n1 = GetMaterialAttributes(None=FunctionInput() {inputName=Material, inputType=FunctionInput_MaterialAttributes}) {attributeGetTypes [#1]}
  n2 = TextureObjectParameter() {group=Military Man, parameterName=Base Diffuse Texture, texture=/Game/Weapons/WeaponSkins/Wraps/Textures/MilitaryMan/MilitaryMan_D}
  n3 = ScalarParameter() {defaultValue=80, group=Military Man, parameterName=Texture Size}
  n4 = FunctionInput() {inputName=Local Position}
  n5 = FunctionInput() {inputName=Local Normal}
  n6 = Constant() {r=1}
  n7 = MaterialFunctionCall(Float4=MaterialFunctionCall(
    Export Float 4=StaticBool() {value=true}
    TextureObject=@n2
    TextureSize=@n3
    WorldPosition=@n4
    World Space Normal=@n5
    ProjectionTransitionContrast=@n6
   ) {materialFunction=&/WorldAlignedTexture}.XYZ Texture) {materialFunction=&BreakOutFloat4Components}
  n8 = MaterialFunctionCall(X=@n7.R, Y=@n7.R, Z=@n7.R) {materialFunction=&MakeFloat3}
  n9 = VectorParameter() {group=Military Man, parameterName=Sun Dried Wear, defaultValue=<rgba 8 0.1498 0 1>}
  n10 = Divide(A=@n4, B=FunctionInput() {inputName=Local Bounds})
  n11 = ComponentMask(None=@n10) {b=true}
  n12 = Power(Base=@n11) {constExponent=3}
  n13 = TextureObjectParameter() {group=Military Man, parameterName=Base Normal Texture, samplerType=SAMPLERTYPE_Normal, texture=/Engine/EngineMaterials/FlatNormal}
  n14 = MaterialFunctionCall!(TextureObject=@n13, TextureSize=@n3, RotYAxis, RotZAxis, RotXAxis, PreSkinNormal=@n5, ProjectionTransitionContrast=@n6, WorldPosition=@n4) {materialFunction=/Game/Packages/Fortress_SharedMaterials/Base_Material_Functions/BaseCustomizationFunctions/PreSkinnedAlignedNormal}
  n15 = ScalarParameter() {defaultValue=0.5, group=Military Man, parameterName=Military Man Normal}
  n16 = SetMaterialAttributes!(
   MaterialAttributes=@n1.MaterialAttributes
   BaseColor=LinearInterpolate(A=@n8.Result, B=Multiply(A=@n8.Result, B=@n9.RGB), Alpha=Saturate(None=@n12))
   Roughness=LinearInterpolate(
    A=ScalarParameter() {group=Military Man, parameterName=Military Man Roughness Min}
    B=ScalarParameter() {defaultValue=1, group=Military Man, parameterName=Military Man Roughness Max}
    Alpha=@n7.A
   )
   Normal=MaterialFunctionCall(
    BaseNormal=@n1.Normal
    AdditionalNormal=MaterialFunctionCall(Normal=@n14.XYZ Texture, Flatness=@n15) {materialFunction=&FlattenNormal}.Result
   ) {materialFunction=&BlendAngleCorrectedNormals}.Result
   Metallic=ScalarParameter() {group=Military Man, parameterName=Military Man Metal}
   Specular=ScalarParameter() {defaultValue=0.8, group=Military Man, parameterName=Military Man Specular}
  ) {attributeSetTypes [
    #2
    #3
    #1
    #4
    #5
   ]}
  n17 = FunctionInput() {bUsePreviewValueAsDefault=true, inputName=Customization Mask, inputType=FunctionInput_Scalar, previewValue=<1 0 0 1>}
  FunctionOutput(None=StaticSwitchParameter(True=BlendMaterialAttributes(A=@n1.MaterialAttributes, B=@n16, Alpha=@n17), False=@n1.MaterialAttributes) {group=Customization, parameterName=UseMilitaryMan})
  LinearInterpolate(
   A=@n11
   B=ComponentMask(None=@n10) {r=true}
   Alpha=FunctionInput(Preview=Constant() {r=1}) {inputName=isVehicle, inputType=FunctionInput_Scalar}
  )
 ]}
