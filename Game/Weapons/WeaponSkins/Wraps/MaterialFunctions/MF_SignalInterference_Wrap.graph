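# Signal Interference wrap material function for weapons, vehicles and mechs, behind a static switch.
# Customization mask blends in dark surface parameters and added emissive.
# MF_WrapBounds supplies UVs, blended by isVehicle/isMech, for two-color edge gradients plus logo, animated front and back glitch layers, masked by local-space projection.
# For EB_ReactiveDuration after 00-GameTimeFromBP, fading UV warp and offset, tinted chromatic-split copies intensify the glitch.
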MaterialFunction /Game/Weapons/WeaponSkins/Wraps/MaterialFunctions/MF_SignalInterference_Wrap
graph {nodes [
  n1 = Multiply(A=Constant() {r=0.35}, B=Time())
  n2 = Sine(None=@n1) {period=1.6}
  n3 = Sine(None=@n1) {period=1.42}
  n4 = Ceil(None=AppendVector(A=AppendVector(A=@n2, B=@n3), B=Sine(None=@n1) {period=1.51}))
  n5 = Constant() {r=0.5}
  n6 = Add(A=Sine(None=@n1) {period=2.1}, B=Multiply(B=@n5) {constA=0.1})
  n7 = Add(A=Sine(None=@n1) {period=0.72}, B=Multiply(B=@n5) {constA=0.25}) {constB=0.75}
  n8 = Saturate(None=Multiply(A=Multiply(A=@n6, B=@n7), B=@n7))
  n9 = Add(A=Constant(), B=Multiply(A=Multiply(A=@n4, B=@n8), B=Constant() {r=5}))
  ComponentMask(None=Desaturation(None=@n9)) {g=true}
  FunctionInput(Preview=VertexInterpolator(VS=PreSkinnedLocalBounds().Half Extents).PS) {bUsePreviewValueAsDefault=true, inputName=Local Bounds}
  n10 = FunctionInput(Preview=MakeMaterialAttributes!(
    BaseColor
    Metallic
    Specular
    Roughness
    Anisotropy
    EmissiveColor
    Opacity
    OpacityMask
    Normal
    Tangent
    WorldPositionOffset
    SubsurfaceColor
    ClearCoat
    ClearCoatRoughness
    AmbientOcclusion
    Refraction
    CustomizedUVs_0
    CustomizedUVs_1
    CustomizedUVs_2
    CustomizedUVs_3
    CustomizedUVs_4
    CustomizedUVs_5
    CustomizedUVs_6
    CustomizedUVs_7
    PixelDepthOffset
    ShadingModel
    Displacement
   )) {bUsePreviewValueAsDefault=true, inputName=Material, inputType=FunctionInput_MaterialAttributes, sortPriority=-200}
  n11 = GetMaterialAttributes(None=@n10) {attributeGetTypes [#1]}
  n12 = VectorParameter() {group=Signal_Interference, parameterName=SI_BaseColor, sortPriority=-100, defaultValue=<rgba 0.0044 0.0044 0.0048 1>}
  n13 = SetMaterialAttributes!(
   MaterialAttributes=@n11.MaterialAttributes
   BaseColor=@n12.RGB
   Metallic=ScalarParameter() {group=Signal_Interference, parameterName=SI_Metallic, sortPriority=-100}
   Specular=ScalarParameter() {defaultValue=0.7, group=Signal_Interference, parameterName=SI_Specular, sortPriority=-100}
   Roughness=ScalarParameter() {defaultValue=0.35, group=Signal_Interference, parameterName=SI_Roughness, sortPriority=-100}
  ) {attributeSetTypes [
    #2
    #3
    #4
    #5
   ]}
  n14 = FunctionInput() {bUsePreviewValueAsDefault=true, inputName="Customization Mask 1 (_CM B Top Left)", inputType=FunctionInput_Scalar, sortPriority=25}
  n15 = Saturate(None=@n14)
  n16 = GetMaterialAttributes(None=BlendMaterialAttributes(A=@n11.MaterialAttributes, B=@n13, Alpha=@n15)) {attributeGetTypes [#1]}
  n17 = FunctionInput(Preview=Constant()) {inputName=isVehicle, inputType=FunctionInput_Scalar}
  n18 = FunctionInput() {bUsePreviewValueAsDefault=true, inputName=Object Bounds Max Adjustment, sortPriority=13}
  n19 = FunctionInput() {bUsePreviewValueAsDefault=true, inputName=Object Bounds Min Adjustment, sortPriority=12}
  n20 = FunctionInput() {bUsePreviewValueAsDefault=true, inputName=Object Bounds Center Adjustment, sortPriority=14}
  n21 = FunctionInput(Preview=Constant()) {bUsePreviewValueAsDefault=true, inputName=isMech, inputType=FunctionInput_Scalar}
  n22 = FunctionInput(Preview=VertexInterpolator(VS=PreSkinnedPosition()).PS) {bUsePreviewValueAsDefault=true, inputName=Pre-Skinned Local Position, sortPriority=41}
  n23 = MaterialFunctionCall(Color=Multiply(
    A=MaterialFunctionCall!(
     is Vehicle=@n17
     Object Bounds Max Adjustment=@n18
     Object Bounds Min Adjustment=@n19
     Object Bounds Center Adjustment=@n20
     isMech=@n21
     Side Project Vehicle
     Local Position=@n22
    ) {materialFunction=&/MF_WrapBounds}.Result
    B=Constant3Vector() {constant=<rgba 1 1 -1 0>}
   )) {materialFunction=&SplitComponents}
  n24 = OneMinus(None=@n23.B)
  n25 = LinearInterpolate(A=AppendVector(A=@n23.G, B=@n23.G), B=AppendVector(A=@n23.R, B=@n23.R), Alpha=@n17)
  n26 = LinearInterpolate(A=@n25, B=Multiply(A=AppendVector(A=@n24, B=@n23.R)) {constB=0.425}, Alpha=@n21)
  n27 = Power(Base=Multiply(A=MaterialFunctionCall(UV Channel=@n26) {materialFunction=&LinearGradient}.UGradient) {constB=0.8}) {constExponent=7.5}
  n28 = VectorParameter() {group=Signal_Interference, parameterName=SI_Emissive01_Color, sortPriority=-10, defaultValue=<rgba 0.7 0.0936 0.4887 1>}
  n29 = MaterialFunctionCall(UV Channel=@n26) {materialFunction=&LinearGradient}
  n30 = Power(Base=Multiply(A=OneMinus(None=@n29.UGradient)) {constB=0.8}) {constExponent=7.5}
  n31 = VectorParameter() {group=Signal_Interference, parameterName=SI_Emissive02_Color, sortPriority=-9, defaultValue=<rgba 0.0116 0.4969 0.2307 1>}
  n32 = MaterialFunctionCall(Float3=@n22) {materialFunction=&BreakOutFloat3Components}
  n33 = MaterialFunctionCall(X=@n32.B, Y=@n32.B, Z=@n32.B) {materialFunction=&MakeFloat3}
  n34 = FunctionInput(Preview=VertexInterpolator(VS=PreSkinnedNormal()).PS) {bUsePreviewValueAsDefault=true, inputName=Pre-Skinned Local Normal, sortPriority=42}
  n35 = MaterialFunctionCall(Float3=@n34) {materialFunction=&BreakOutFloat3Components}
  n36 = MaterialFunctionCall(X=@n35.B, Y=@n35.B, Z=@n35.B) {materialFunction=&MakeFloat3}
  n37 = MaterialFunctionCall!(
   WorldPosition=LinearInterpolate(
    A=LinearInterpolate(A=@n22, B=@n33.Result, Alpha=@n17)
    B=MaterialFunctionCall(X=@n32.R, Y=@n32.R, Z=@n32.R) {materialFunction=&MakeFloat3}.Result
    Alpha=@n21
   )
   World Space Normal=LinearInterpolate(
    A=LinearInterpolate(A=@n34, B=@n36.Result, Alpha=@n17)
    B=MaterialFunctionCall(X=@n35.R, Y=@n35.R, Z=@n35.R) {materialFunction=&MakeFloat3}.Result
    Alpha=@n21
   )
   UVSize_Front
   UVSize_Side
   UVSize_Top
   ProjectionTransitionContrast=Constant() {r=0.5}
  ) {materialFunction=/Game/Packages/Fortress_SharedMaterials/Base_Material_Functions/WorldAlignedTexture_UVsMasks}
  n38 = VectorParameter() {group=Signal_Interference, parameterName=SI_Logo_WeaponOffset, primitiveDataIndex=13, defaultValue=<rgba 0 0 0 1>}
  n39 = VectorParameter() {group=Signal_Interference, parameterName=SI_Logo_VehicleOffset, primitiveDataIndex=14, sortPriority=3, defaultValue=<rgba 0 0 0 1>}
  n40 = VectorParameter() {group=Signal_Interference, parameterName=SI_Logo_MechOffset, primitiveDataIndex=15, sortPriority=6, defaultValue=<rgba 0 0 0 1>}
  n41 = ComponentMask(None=LinearInterpolate(A=LinearInterpolate(A=@n38.RGB, B=@n39.RGB, Alpha=@n17), B=@n40.RGB, Alpha=@n21)) {g=true, r=true}
  n42 = LinearInterpolate(A=@n25, B=AppendVector(A=OneMinus(None=@n23.B), B=@n23.R), Alpha=@n21)
  n43 = MaterialFunctionCall!(UV=@n42) {materialFunction=/Game/Weapons/WeaponSkins/MaterialFunction/MF_GlitchFunction}
  n44 = LinearInterpolate(
   A=ScalarParameter() {defaultValue=2, group=Signal_Interference, parameterName=SI_WarpUVS_TileRateWeapon, sortPriority=53}
   B=ScalarParameter() {defaultValue=2, group=Signal_Interference, parameterName=SI_WarpUVS_TileRateVehicle, sortPriority=54}
   Alpha=@n17
  )
  n45 = Time() {period=1}
  n46 = ComponentMask(None=LinearInterpolate(A=Constant3Vector() {constant=<rgba 0.25 0.125 0 1>}, B=Constant3Vector() {constant=<rgba 0.25 0.125 0 1>}, Alpha=@n17)) {g=true, r=true}
  n47 = TextureObjectParameter() {group=Signal_Interference, parameterName=SI_TextureMask03, sortPriority=-1, texture=&T_BlacksRGB}
  n48 = TextureSample(UVs=Add(A=Multiply(A=@n43.Glitched UV, B=@n44), B=Multiply(A=@n45, B=@n46)), Tex=@n47)
  n49 = LinearInterpolate(
   A=ScalarParameter() {defaultValue=0.25, group=Signal_Interference, parameterName=SI_WarpUVS_Amount Weapon, sortPriority=55}
   B=ScalarParameter() {defaultValue=0.25, group=Signal_Interference, parameterName=SI_WarpUVS_Amount Vehicle, sortPriority=56}
   Alpha=@n17
  )
  n50 = Subtract(A=Time(), B=ScalarParameter() {group=00-Reactive-GTFB, parameterName=00-GameTimeFromBP})
  n51 = ScalarParameter() {defaultValue=1.5, group=Electric Burst, parameterName=EB_ReactiveDuration}
  n52 = OneMinus(None=Saturate(None=Divide(A=@n50, B=@n51)))
  n53 = LinearInterpolate(A=@n42, B=Add(A=@n42, B=Multiply(A=@n48.B, B=@n49)), Alpha=@n52)
  n54 = LinearInterpolate(
   A=ScalarParameter() {defaultValue=1, group=Signal_Interference, parameterName=SI_Logo_WeaponScale, sortPriority=1}
   B=ScalarParameter() {defaultValue=1, group=Signal_Interference, parameterName=SI_Logo_VehicleScale, sortPriority=4}
   Alpha=@n17
  )
  n55 = ScalarParameter() {defaultValue=1, group=Signal_Interference, parameterName=SI_Logo_MechScale, sortPriority=7}
  n56 = VectorParameter() {group=Signal_Interference, parameterName=SI_Logo_RotationWeapon, sortPriority=2, defaultValue=<rgba 0 0 -0.25 1>}
  n57 = VectorParameter() {group=Signal_Interference, parameterName=SI_Logo_RotationVehicle, sortPriority=5, defaultValue=<rgba 0 0 0 1>}
  n58 = VectorParameter() {group=Signal_Interference, parameterName=SI_Logo_RotationMech, sortPriority=8, defaultValue=<rgba 0.25 0.25 0 1>}
  n59 = MaterialFunctionCall(
   UVs=Multiply(A=Add(A=@n53, B=@n41), B=LinearInterpolate(A=@n54, B=@n55, Alpha=@n21))
   Rotation Center=Constant2Vector() {g=0.5, r=0.5}
   "Rotation Angle (0-1)"=LinearInterpolate(A=LinearInterpolate(A=@n56.RGB, B=@n57.RGB, Alpha=@n17), B=@n58.RGB, Alpha=@n21)
  ) {materialFunction=&CustomRotator}
  n60 = TextureObjectParameter() {group=Signal_Interference, parameterName=SI_TextureMask01, sortPriority=-1, texture=&T_BlacksRGB}
  n61 = TextureSample(UVs=@n59.Rotated Values, Tex=@n60) {samplerType=SAMPLERTYPE_Masks}
  n62 = Multiply(A=@n61.R, B=@n28.RGB)
  n63 = OneMinus(None=@n61.R)
  n64 = Constant() {r=1}
  n65 = MaterialFunctionCall(UV=@n59.Rotated Values) {materialFunction=&MF_GlitchFunction}
  n66 = Add(A=AppendVector(A=Constant() {r=1.05}, B=@n64), B=@n65.Glitched UV)
  n67 = Multiply(
   A=Multiply(A=@n63, B=TextureSample(UVs=@n66, Tex=@n60) {samplerType=SAMPLERTYPE_Masks}.R)
   B=Constant3Vector() {constant=<rgba 0.9131 0 0.013 1>}
  )
  n68 = Add(A=@n65.Glitched UV, B=AppendVector(A=Constant() {r=0.95}, B=@n64))
  n69 = Multiply(
   A=Multiply(A=@n63, B=TextureSample(UVs=@n68, Tex=@n60) {samplerType=SAMPLERTYPE_Masks}.R)
   B=Constant3Vector() {constant=<rgba 0.0015 0.0908 0.5647 1>}
  )
  n70 = LinearInterpolate(A=@n62, B=Add(A=@n62, B=Add(A=@n67, B=@n69)), Alpha=@n52)
  n71 = Time()
  n72 = Constant2Vector() {g=0.075}
  n73 = Multiply(A=Add(A=@n53, B=Multiply(A=@n71, B=@n72)), B=Constant() {r=0.25})
  n74 = Time()
  n75 = Constant2Vector() {g=-0.05}
  n76 = Multiply(A=Add(A=@n53, B=Multiply(A=@n74, B=@n75)), B=Constant() {r=1.5})
  n77 = Multiply(A=Add(A=TextureSample(UVs=@n73, Tex=@n47).G, B=TextureSample(UVs=@n76, Tex=@n47).B), B=TextureSample(UVs=@n53, Tex=@n47).R)
  n78 = ScalarParameter() {defaultValue=2, group=Signal_Interference, parameterName=SI_WarpUVS_Front_TileRateWeapon, sortPriority=30}
  n79 = ScalarParameter() {defaultValue=2, group=Signal_Interference, parameterName=SI_WarpUVS_Front_TileRateVehicle, sortPriority=31}
  n80 = Multiply(
   A=@n53
   B=LinearInterpolate(
    A=LinearInterpolate(A=@n78, B=@n79, Alpha=@n17)
    B=ScalarParameter() {defaultValue=2, group=Signal_Interference, parameterName=SI_WarpUVS_Front_TileRateMech}
    Alpha=@n21
   )
  )
  n81 = VectorParameter() {group=Signal_Interference, parameterName=SI_Warp_Front_AnimationWeapon, sortPriority=30, defaultValue=<rgba 0.25 0.01 0 1>}
  n82 = VectorParameter() {group=Signal_Interference, parameterName=SI_Warp_Front_AnimationVehicle, sortPriority=31, defaultValue=<rgba 0.1 0.01 0 1>}
  n83 = LinearInterpolate(A=@n81.RGB, B=@n82.RGB, Alpha=@n17)
  n84 = VectorParameter() {group=Signal_Interference, parameterName=SI_Warp_Front_AnimationMech, defaultValue=<rgba 0.1 0.01 0 1>}
  n85 = Multiply(A=Time(), B=ComponentMask(None=LinearInterpolate(A=@n83, B=@n84.RGB, Alpha=@n21)) {g=true, r=true})
  n86 = TextureObjectParameter() {group=Signal_Interference, parameterName=SI_TextureMask04, sortPriority=-1, texture=&T_BlacksRGB}
  n87 = ScalarParameter() {defaultValue=0.25, group=Signal_Interference, parameterName=SI_WarpUVS_Front_Amount Weapon, sortPriority=30}
  n88 = ScalarParameter() {defaultValue=0.25, group=Signal_Interference, parameterName=SI_WarpUVS_Front_Amount Vehicle, sortPriority=31}
  n89 = Multiply(
   A=TextureSample(UVs=Add(A=@n80, B=@n85), Tex=@n86).G
   B=LinearInterpolate(
    A=LinearInterpolate(A=@n87, B=@n88, Alpha=@n17)
    B=ScalarParameter() {defaultValue=0.25, group=Signal_Interference, parameterName=SI_WarpUVS_Front_Amount Mech}
    Alpha=@n21
   )
  )
  n90 = VectorParameter() {group=Signal_Interference, parameterName=SI_Front_WeaponOffset, primitiveDataIndex=13, sortPriority=20, defaultValue=<rgba 0 0 0 1>}
  n91 = VectorParameter() {group=Signal_Interference, parameterName=SI_Front_VehicleOffset, primitiveDataIndex=14, sortPriority=21, defaultValue=<rgba 0 0 0 1>}
  n92 = LinearInterpolate(A=@n90.RGB, B=@n91.RGB, Alpha=@n17)
  n93 = VectorParameter() {group=Signal_Interference, parameterName=SI_Front_MechOffset, primitiveDataIndex=15, sortPriority=22, defaultValue=<rgba 0 0 0 1>}
  n94 = Add(A=@n53, B=ComponentMask(None=LinearInterpolate(A=@n92, B=@n93.RGB, Alpha=@n21)) {g=true, r=true})
  n95 = LinearInterpolate(
   A=LinearInterpolate(
    A=ScalarParameter() {defaultValue=1.5, group=Signal_Interference, parameterName=SI_Front_WeaponScale, sortPriority=20}
    B=ScalarParameter() {defaultValue=1.5, group=Signal_Interference, parameterName=SI_Front_VehicleScale, sortPriority=21}
    Alpha=@n17
   )
   B=ScalarParameter() {defaultValue=1.5, group=Signal_Interference, parameterName=SI_Front_MechScale, sortPriority=22}
   Alpha=@n21
  )
  n96 = VectorParameter() {group=Signal_Interference, parameterName=SI_Front_RotationWeapon, sortPriority=20, defaultValue=<rgba 0 0 0 1>}
  n97 = VectorParameter() {group=Signal_Interference, parameterName=SI_Front_RotationVehicle, sortPriority=21, defaultValue=<rgba 0 0 0 1>}
  n98 = VectorParameter() {group=Signal_Interference, parameterName=SI_Front_RotationMech, sortPriority=22, defaultValue=<rgba 0 0 0 1>}
  n99 = MaterialFunctionCall(
   UVs=Add(A=Multiply(A=@n94, B=@n95), B=@n89)
   Rotation Center=Constant2Vector() {g=0.5, r=0.5}
   "Rotation Angle (0-1)"=LinearInterpolate(A=LinearInterpolate(A=@n96.RGB, B=@n97.RGB, Alpha=@n17), B=@n98.RGB, Alpha=@n21)
  ) {materialFunction=&CustomRotator}
  n100 = ScalarParameter() {defaultValue=1, group=Signal_Interference, parameterName=SI_Front_HeightLerpContrast, sortPriority=35}
  n101 = VectorParameter() {
   desc=GlitchMask 01 texture becomes Emissive 01
   group=Signal_Interference
   parameterName=SI_Emissive01_Color
   sortPriority=-10
   defaultValue=<rgba 0.9734 0.1301 0.6795 1>
  }
  n102 = Multiply(
   A=MaterialFunctionCall(Transition Phase=@n89, Height Texture=TextureSample(UVs=@n99.Rotated Values, Tex=@n86).R, Contrast=@n100) {materialFunction=&HeightLerp}.Alpha
   B=@n101.RGB
  )
  n103 = Constant() {r=1}
  n104 = Add(A=AppendVector(A=Constant() {r=1.05}, B=@n103), B=@n99.Rotated Values)
  n105 = Multiply(
   A=MaterialFunctionCall(Transition Phase=@n89, Height Texture=TextureSample(UVs=@n104, Tex=@n86).R, Contrast=@n100) {materialFunction=&HeightLerp}.Alpha
   B=Constant3Vector() {constant=<rgba 0.9131 0 0.013 1>}
  )
  n106 = Add(A=@n99.Rotated Values, B=AppendVector(A=Constant() {r=0.95}, B=@n103))
  n107 = Multiply(
   A=MaterialFunctionCall(Transition Phase=@n89, Height Texture=TextureSample(UVs=@n106, Tex=@n86).R, Contrast=@n100) {materialFunction=&HeightLerp}.Alpha
   B=Constant3Vector() {constant=<rgba 0.0015 0.0908 0.5647 1>}
  )
  n108 = Add(A=Add(A=Multiply(A=@n105), B=Multiply(A=@n107)), B=@n102)
  n109 = VectorParameter() {group=Signal_Interference, parameterName=SI_GlitchBack_WeaponOffset, primitiveDataIndex=13, sortPriority=40, defaultValue=<rgba 0 0 0 1>}
  n110 = VectorParameter() {group=Signal_Interference, parameterName=SI_GlitchBack_VehicleOffset, primitiveDataIndex=14, sortPriority=41, defaultValue=<rgba 0 0 0 1>}
  n111 = VectorParameter() {group=Signal_Interference, parameterName=SI_GlitchBack_MechOffset, primitiveDataIndex=15, sortPriority=42, defaultValue=<rgba 0 0 0 1>}
  n112 = ComponentMask(None=LinearInterpolate(A=LinearInterpolate(A=@n109.RGB, B=@n110.RGB, Alpha=@n17), B=@n111.RGB, Alpha=@n21)) {g=true, r=true}
  n113 = LinearInterpolate(
   A=ScalarParameter() {defaultValue=1, group=Signal_Interference, parameterName=SI_GlitchBack_WeaponScale, sortPriority=40}
   B=ScalarParameter() {defaultValue=1, group=Signal_Interference, parameterName=SI_GlitchBack_VehicleScale, sortPriority=41}
   Alpha=@n17
  )
  n114 = ScalarParameter() {defaultValue=1, group=Signal_Interference, parameterName=SI_GlitchBack_MechScale, sortPriority=42}
  n115 = VectorParameter() {group=Signal_Interference, parameterName=SI_Back_RotationWeapon, sortPriority=40, defaultValue=<rgba 0 0 0 1>}
  n116 = VectorParameter() {group=Signal_Interference, parameterName=SI_Back_RotationVehicle, sortPriority=41, defaultValue=<rgba 0 0 0 1>}
  n117 = VectorParameter() {group=Signal_Interference, parameterName=SI_Back_RotationMech, sortPriority=42, defaultValue=<rgba 0 0 0 1>}
  n118 = MaterialFunctionCall(
   UVs=Multiply(A=Add(A=@n53, B=@n112), B=LinearInterpolate(A=@n113, B=@n114, Alpha=@n21))
   Rotation Center=Constant2Vector() {g=0.5, r=0.5}
   "Rotation Angle (0-1)"=LinearInterpolate(A=LinearInterpolate(A=@n115.RGB, B=@n116.RGB, Alpha=@n17), B=@n117.RGB, Alpha=@n21)
  ) {materialFunction=&CustomRotator}
  n119 = TextureObjectParameter() {group=Signal_Interference, parameterName=SI_TextureMask02, sortPriority=-1, texture=&T_BlacksRGB}
  n120 = TextureObjectParameter() {group=Signal_Interference, parameterName=SI_TextureMask01, sortPriority=-1, texture=&T_BlacksRGB}
  n121 = MaterialFunctionCall!(Bounds=@n118.Rotated Values, Texture_1=@n119, Texture=@n120) {materialFunction=&/MF_SignalInterference_Back}
  n122 = Multiply(A=@n31.RGB, B=@n121.Back_UV)
  n123 = Constant2Vector() {g=1.01, r=1}
  n124 = MaterialFunctionCall(
   Bounds=MaterialFunctionCall(UV=Add(A=@n118.Rotated Values, B=@n123)) {materialFunction=&MF_GlitchFunction}.Glitched UV
   Texture_1=@n119
   Texture=@n120
  ) {materialFunction=&MF_SignalInterference_Back}
  n125 = Constant3Vector() {constant=<rgba 0.011 0.4621 0.0242 1>}
  n126 = Constant2Vector() {g=0.95, r=1}
  n127 = MaterialFunctionCall(
   Bounds=MaterialFunctionCall(UV=Add(A=@n118.Rotated Values, B=@n126)) {materialFunction=&MF_GlitchFunction}.Glitched UV
   Texture_1=@n119
   Texture=@n120
  ) {materialFunction=&MF_SignalInterference_Back}
  n128 = Constant3Vector() {constant=<rgba 0.0015 0.0908 0.5647 1>}
  n129 = Multiply(A=Add(A=Multiply(A=@n124.Back_UV, B=@n125), B=Multiply(A=@n127.Back_UV, B=@n128)), B=OneMinus(None=@n121.Back_UV))
  n130 = Add(
   A=LinearInterpolate(A=Multiply(A=@n77, B=@n102), B=Multiply(A=@n77, B=@n108), Alpha=@n52)
   B=LinearInterpolate(A=@n122, B=Add(A=@n129, B=@n122), Alpha=@n52)
  )
  n131 = Add(A=Add(A=Multiply(A=@n27, B=@n28.RGB), B=Multiply(A=@n30, B=@n31.RGB)), B=Multiply(A=OneMinus(None=@n37.Mask_Front), B=Add(A=@n70, B=@n130)))
  n132 = Clamp(None=LinearInterpolate(A=@n11.EmissiveColor, B=Add(A=@n11.EmissiveColor, B=@n131), Alpha=@n15)) {maxDefault=100}
  FunctionOutput(None=StaticSwitchParameter(
    True=SetMaterialAttributes!(MaterialAttributes=@n16.MaterialAttributes, EmissiveColor=@n132) {attributeSetTypes [#1]}
    False=@n10
   ) {group=Customization_Wrap, parameterName=Wrap_SignalInterference})
  n133 = Multiply(A=@n48.B, B=@n9)
  Power(Base=Desaturation(None=Saturate(None=@n133))) {constExponent=3}
  Reroute(None=@n119)
  ScalarParameter() {desc="0 = effect is off, 1 = effect is on", group=TopSideProjection_Reactive, parameterName=TurnOnEmissive_ForSequencer}
 ]}
