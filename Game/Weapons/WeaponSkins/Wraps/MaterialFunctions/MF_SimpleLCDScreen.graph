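# Material function for weapon wraps that turns a customization-mask channel into an emissive LCD screen, behind a static switch.
# Panning sphere-mask scanlines modulate the screen color, then add a sine-blinked crosshair from the screen texture's green channel, with red masking the emission.
# HUD UVs offset and scale about center, using only the U bias and scale for both axes.
# Emissive and surface overrides blend into incoming material attributes by that mask.
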MaterialFunction /Game/Weapons/WeaponSkins/Wraps/MaterialFunctions/MF_SimpleLCDScreen
graph {nodes [
  n1 = NamedRerouteUsage()
  n2 = VectorParameter() {group=SimpleLCDScreen, parameterName=ScreenColor, defaultValue=<rgba 0 0 1 0>}
  n3 = TextureCoordinate()
  n4 = ScalarParameter() {defaultValue=10, group=SimpleLCDScreen, parameterName=ScanlineTiling}
  n5 = Constant()
  n6 = ScalarParameter() {defaultValue=1, group=SimpleLCDScreen, parameterName=ScanlinePanSpeed_V}
  n7 = ComponentMask(None=Panner(Coordinate=Multiply(A=@n3, B=@n4), Speed=AppendVector(A=@n5, B=@n6)) {bFractionalPart=true}) {g=true}
  n8 = Add(
   A=SphereMask(
    A=Frac(None=@n7)
    B=Constant() {r=0.5}
    Radius=ScalarParameter() {defaultValue=0.3, group=SimpleLCDScreen, parameterName=ScanlineRadius}
    Hardness=ScalarParameter() {defaultValue=0.8, group=SimpleLCDScreen, parameterName=ScanlineHardness}
   )
   B=ScalarParameter() {defaultValue=-0.5, group=SimpleLCDScreen, parameterName=ScanlineBias}
  )
  n9 = ScalarParameter() {defaultValue=0.2, group=SimpleLCDScreen, parameterName=ScanlineScale}
  n10 = Add(A=@n2.RGB, B=Multiply(A=@n2.RGB, B=Multiply(A=@n8, B=@n9)))
  n11 = VectorParameter() {group=SimpleLCDScreen, parameterName=HUD_UV_Bias, channelNames {a=V Scale, b=U Scale, g=V Bias, r=U Bias}, defaultValue=<rgba 0 0 1 1>}
  n12 = Add(A=TextureCoordinate(), B=AppendVector(A=@n11.U Bias, B=@n11.U Bias))
  n13 = AppendVector(A=@n11.U Scale, B=@n11.U Scale)
  n14 = TextureSample(UVs=Saturate(None=MaterialFunctionCall(UVs=@n12, Texture Scale=@n13) {materialFunction=&ScaleUVsByCenter}.UVs)) {texture=/Game/Weapons/WeaponSkins/Wraps/Textures/T_ArtilleryCannon_Screen}
  n15 = Sine(None=Time()) {period=0.85}
  n16 = Multiply(
   A=Multiply(A=VectorParameter() {parameterName=CrosshairColor, defaultValue=<rgba 1 0.297 0 0>}.RGB, B=@n14.G)
   B=Ceil(None=Saturate(None=@n15))
  )
  n17 = SetMaterialAttributes!(
   MaterialAttributes=NamedRerouteUsage()
   EmissiveColor=Multiply(A=Add(A=@n10, B=@n16), B=@n14.R)
   BaseColor=VectorParameter() {group=SimpleLCDScreen, parameterName=BaseColorOverride}.RGB
   Metallic=ScalarParameter() {group=SimpleLCDScreen, parameterName=MetallicOverride}
   Specular=ScalarParameter() {group=SimpleLCDScreen, parameterName=SpecularOverride}
   Roughness=ScalarParameter() {group=SimpleLCDScreen, parameterName=RoughnessOverride}
  ) {attributeSetTypes [
    #1
    #2
    #3
    #4
    #5
   ]}
  n18 = ChannelMaskParameter(None=FunctionInput() {inputName=CustomizationMask, inputType=FunctionInput_Vector4, sortPriority=1}) {group=SimpleLCDScreen, maskChannel=Blue, parameterName=CustomizationMaskChannel, defaultValue=<rgba 0 0 1 0>}
  FunctionOutput(None=StaticSwitchParameter(True=BlendMaterialAttributes(A=@n1, B=@n17, Alpha=@n18.R) {vertexAttributeBlendType=UseA}, False=NamedRerouteUsage()) {group=SimpleLCDScreen, parameterName=UseSimpleLCDScreen})
  NamedRerouteDeclaration(None=FunctionInput() {inputName=MA, inputType=FunctionInput_MaterialAttributes}) {name=MA_In, nodeColor=<rgba 1 0.5156 0 1>}
 ]}
