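# Material function that adds glinting sparkles to weapon, vehicle and character wrap skins, toggled by a static switch.
# Two pre-skinned-position triplanar noise samples, offset oppositely by time or camera direction, are thresholded and multiplied into sparse glints.
# Glints alter base color, roughness and emissive, with camera-distance fade on color and glow, blended through a customization mask.
# Cheap-triplanar switches and reflection-vector noise branches are left unconnected to the output.
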MaterialFunction /Game/Weapons/WeaponSkins/Wraps/MaterialFunctions/MF_Sparkeles
graph {nodes [
  CameraPositionWS()
  n1 = FunctionInput() {inputName=Material, inputType=FunctionInput_MaterialAttributes}
  n2 = GetMaterialAttributes(None=@n1) {attributeGetTypes [#1, #2, #3]}
  n3 = TextureObjectParameter() {
   group=Sparkles
   parameterName=Sparkles_textures
   samplerType=SAMPLERTYPE_LinearColor
   texture=/Game/Global/Textures/Default/Noise/T_Wrap_SnowStreak_Noise
  }
  n4 = MaterialFunctionCall(Float4=FunctionInput() {inputName=isState, inputType=FunctionInput_Vector4, sortPriority=4}) {materialFunction=&BreakOutFloat4Components}
  n5 = LinearInterpolate(
   A=LinearInterpolate(
    A=ScalarParameter() {defaultValue=8, group=Sparkles, parameterName=Sparkles_Tile_RateWeapon}
    B=ScalarParameter() {defaultValue=24, group=Sparkles, parameterName=Sparkles_Tile_RateVehicle}
    Alpha=@n4.R
   )
   B=ScalarParameter() {defaultValue=15, group=Sparkles, parameterName=Sparkles_Tile_RateCharacter}
   Alpha=@n4.A
  )
  n6 = FunctionInput(Preview=VertexInterpolator(VS=PreSkinnedPosition()).PS) {bUsePreviewValueAsDefault=true, inputName=Local Position, sortPriority=1}
  n7 = Multiply(A=@n6, B=Constant() {r=0.5})
  n8 = MaterialFunctionCall() {materialFunction=&CameraDirectionVector}
  n9 = VectorParameter() {group=Sparkles, parameterName=Sparkles_Vehicle_Speed, sortPriority=50, defaultValue=<rgba 0 0 0 1>}
  n10 = Constant() {r=1}
  n11 = Multiply(
   A=Time()
   B=LinearInterpolate(
    A=VectorParameter() {group=Sparkles, parameterName=Sparkles_Weapon_Speed, sortPriority=50, defaultValue=<rgba 0 0 0 1>}.RGB
    B=@n9.RGB
    Alpha=FunctionInput(Preview=@n10) {inputName=isVehicle, inputType=FunctionInput_Scalar, sortPriority=4}
   )
  )
  n12 = ScalarParameter() {defaultValue=1, group=Sparkles, parameterName="Sparkles_Animated_Panning --->Camera Based", sortPriority=50}
  n13 = LinearInterpolate(A=Multiply(A=@n8.Result, B=ScalarParameter() {defaultValue=500, group=Sparkles, parameterName=Sparkles_Camera_Movement}), B=@n11, Alpha=@n12)
  n14 = FunctionInput(Preview=VertexInterpolator(VS=PreSkinnedNormal()).PS) {bUsePreviewValueAsDefault=true, inputName=Pre-Skinned Local Normal, sortPriority=2}
  n15 = MaterialFunctionCall(TextureObject=@n3, TextureSize=@n5, WorldPosition=Add(A=@n7, B=@n13), World Space Normal=@n14) {materialFunction=&WorldAlignedTexture}
  n16 = MaterialFunctionCall(Color=@n15.XYZ Texture) {materialFunction=&SplitComponents}
  n17 = Constant() {r=-0.95}
  n18 = Add(A=Saturate(None=Add(A=@n16.B, B=@n17)), B=Saturate(None=Add(A=@n16.G, B=@n17)))
  n19 = Saturate(None=Add(A=@n16.R, B=@n17))
  n20 = MaterialFunctionCall(TextureObject=@n3, TextureSize=OneMinus(None=@n5), WorldPosition=Add(A=@n7, B=OneMinus(None=@n13)), World Space Normal=@n14) {materialFunction=&WorldAlignedTexture}
  n21 = MaterialFunctionCall(Color=@n20.XYZ Texture) {materialFunction=&SplitComponents}
  n22 = Saturate(None=Add(A=@n21.R, B=@n17))
  n23 = Add(A=Saturate(None=Add(A=@n21.G, B=@n17)), B=Saturate(None=Add(A=@n21.B, B=@n17)))
  n24 = Multiply(A=Multiply(A=Add(A=@n18, B=@n19), B=Add(A=@n22, B=@n23)), B=Constant() {r=500}) {constB=3}
  n25 = Subtract(A=Constant() {r=1}, B=Saturate(None=@n24))
  n26 = MaterialFunctionCall(Fade Length=ScalarParameter() {defaultValue=2500, group=Sparkles, parameterName=Saprkle_FadeDistance}) {materialFunction=&CameraDepthFade}
  n27 = MaterialFunctionCall(
   Base=@n2.BaseColor
   Blend=LinearInterpolate(A=LinearInterpolate(Alpha=@n25) {constA=1, constB=0.5}, Alpha=@n26.Result) {constA=0.5, constAlpha=0, constB=0.5}
  ) {materialFunction=&Blend_Overlay}
  n28 = LinearInterpolate(A=LinearInterpolate(B=ScalarParameter() {defaultValue=1, group=Sparkles, parameterName=Saprkle_GlowStrength}, Alpha=@n24), Alpha=@n26.Result) {constA=0.5, constAlpha=0, constB=0}
  n29 = SetMaterialAttributes!(
   MaterialAttributes=@n2.MaterialAttributes
   BaseColor=@n27.Result
   Roughness=Multiply(A=@n2.Roughness, B=@n25)
   EmissiveColor=LinearInterpolate(A=Multiply(A=@n27.Result, B=@n28), B=@n2.EmissiveColor, Alpha=@n25)
  ) {attributeSetTypes [#1, #2, #3]}
  n30 = Saturate(None=Add(
    A=FunctionInput() {bUsePreviewValueAsDefault=true, inputName="Customization Mask (_CM B)", inputType=FunctionInput_Scalar, sortPriority=24}
    B=ScalarParameter() {group=Sparkles, parameterName=Saprkle_Primary_Masking, sortPriority=-4}
   ))
  FunctionOutput(None=StaticSwitchParameter(True=BlendMaterialAttributes(A=@n2.MaterialAttributes, B=@n29, Alpha=@n30), False=@n1) {group=Customization, parameterName=Sparkles})
  n31 = LinearInterpolate(
   A=Multiply(A=@n8.Result, B=ScalarParameter() {defaultValue=500, group=Sparkles, parameterName=Sparkles_Camera_MovementCH})
   B=@n11
   Alpha=@n12
  )
  n32 = MaterialFunctionCall!(Local Position=Add(A=@n6, B=OneMinus(None=@n31)), Pre-Skinned Local Normal=@n14) {materialFunction=&/MF_Cheap_Tri}
  n33 = OneMinus(None=@n5)
  n34 = StaticBoolParameter() {group=Sparkles, parameterName=Cheap_Triplaner, sortPriority=-5}
  StaticSwitch(
   True=TextureSample(UVs=Divide(A=@n32.Result, B=@n33), Tex=@n3) {samplerType=SAMPLERTYPE_LinearColor, texture=&T_Wrap_SnowStreak_Noise}.RGB
   False=@n20.XYZ Texture
   Value=@n34
  )
  n35 = MaterialFunctionCall(Local Position=Add(A=@n6, B=@n31), Pre-Skinned Local Normal=@n14) {materialFunction=&MF_Cheap_Tri}
  StaticSwitch(
   True=TextureSample(UVs=Divide(A=@n35.Result, B=@n5), Tex=@n3) {samplerType=SAMPLERTYPE_LinearColor, texture=&T_Wrap_SnowStreak_Noise}.RGB
   False=@n15.XYZ Texture
   Value=@n34
  )
  n36 = ReflectionVectorWS()
  TransformPosition(None=@n36) {transformSourceType=TRANSFORMPOSSOURCE_World}
  n37 = Constant() {r=100}
  n38 = Divide(A=OneMinus(None=@n36), B=@n37) {constB=15}
  n39 = Multiply(A=@n6) {constB=0.05}
  n40 = ScalarParameter() {defaultValue=32, group=Sparkles, parameterName=Sparkles_Tile_Rate}
  VectorNoise(World Position=Multiply(A=Add(A=@n38, B=@n39), B=@n40) {constB=10})
  n41 = Divide(A=@n36, B=@n37) {constB=15}
  n42 = Multiply(A=@n6) {constB=0.05}
  VectorNoise(World Position=Multiply(A=Add(A=@n41, B=@n42), B=@n40) {constB=10})
 ]}
