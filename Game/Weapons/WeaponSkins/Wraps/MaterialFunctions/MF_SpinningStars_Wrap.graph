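# Material function for the SpinningStars weapon wrap, also covering vehicles and mechs, that blends animated star layers into the incoming material behind a static switch.
# Bounds-based planar UVs drive base and star-color textures, while top/side-projected masks are height-lerped against animated gradients to form primary and secondary stars.
# A timed reactive pulse from a Blueprint-set game time shifts hue, gradient opacity and emissive, revealing secondary-star glow.
# Customization masks limit the final attribute blend.
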MaterialFunction /Game/Weapons/WeaponSkins/Wraps/MaterialFunctions/MF_SpinningStars_Wrap
graph {nodes [
  n1 = FunctionInput() {inputName=Material, inputType=FunctionInput_MaterialAttributes}
  n2 = GetMaterialAttributes(None=@n1) {attributeGetTypes [#1]}
  n3 = FunctionInput() {bUsePreviewValueAsDefault=true, inputName=Wrap Bounds, sortPriority=12}
  n4 = Multiply(A=@n3, B=Constant3Vector() {constant=<rgba 1 1 -1 0>})
  n5 = VectorParameter() {group=SpinningStars, parameterName=SS_BaseTextureOffset_Weapon, sortPriority=15, defaultValue=<rgba 0 0 0 1>}
  n6 = VectorParameter() {group=SpinningStars, parameterName=SS_BaseTextureOffset_Vehicle, sortPriority=15, defaultValue=<rgba 0 0 0 1>}
  n7 = VectorParameter() {group=SpinningStars, parameterName=SS_BaseTextureOffset_Vehicle_Side, sortPriority=15, defaultValue=<rgba 0 0 0 1>}
  n8 = FunctionInput() {bUsePreviewValueAsDefault=true, inputName=Side Project Vehicle, inputType=FunctionInput_Scalar}
  n9 = FunctionInput(Preview=Constant()) {inputName=isVehicle, inputType=FunctionInput_Scalar, sortPriority=5}
  n10 = VectorParameter() {group=SpinningStars, parameterName=SS_BaseTextureOffset_Mech, sortPriority=15, defaultValue=<rgba 0 0 0 1>}
  n11 = FunctionInput(Preview=Constant()) {bUsePreviewValueAsDefault=true, inputName=isMech, inputType=FunctionInput_Scalar, sortPriority=4}
  n12 = LinearInterpolate(A=LinearInterpolate(A=@n5.RGB, B=LinearInterpolate(A=@n6.RGB, B=@n7.RGB, Alpha=@n8), Alpha=@n9), B=@n10.RGB, Alpha=@n11)
  n13 = VectorParameter() {group=SpinningStars, parameterName="SS_BaseScale_Weapon_01 A=uniform", sortPriority=16, defaultValue=<rgba 1 1 1 1>}
  n14 = VectorParameter() {group=SpinningStars, parameterName="SS_BaseScale_Vehicle_01 A=uniform", sortPriority=16, defaultValue=<rgba 1 1 1 1>}
  n15 = VectorParameter() {
   desc=must activate TS_Vehicle_Side_01 UVs
   group=SpinningStars
   parameterName="SS_BaseScale_Vehicle_Side_01 A=uniform"
   sortPriority=16
   defaultValue=<rgba 1 1 1 1>
  }
  n16 = LinearInterpolate(
   A=Multiply(A=@n13.RGB, B=@n13.RGB)
   B=LinearInterpolate(A=Multiply(A=@n14.RGB, B=@n14.RGB), B=Multiply(A=@n15.RGB, B=@n15.RGB), Alpha=@n8)
   Alpha=@n9
  )
  n17 = VectorParameter() {group=SpinningStars, parameterName="SS_BaseScale_Mech_01 A=uniform", sortPriority=16, defaultValue=<rgba 1 1 1 1>}
  n18 = Multiply(A=@n17.RGB, B=@n17.RGB)
  n19 = MaterialFunctionCall(Color=Multiply(A=Add(A=@n4, B=@n12), B=LinearInterpolate(A=@n16, B=@n18, Alpha=@n11))) {materialFunction=&SplitComponents}
  n20 = AppendVector(A=@n19.R, B=@n19.R)
  n21 = AppendVector(A=@n19.R, B=@n19.R)
  n22 = LinearInterpolate(
   A=LinearInterpolate(A=AppendVector(A=@n19.G, B=@n19.G), B=LinearInterpolate(A=@n20, B=@n21, Alpha=@n8), Alpha=@n9)
   B=AppendVector(A=OneMinus(None=@n19.B), B=@n19.R)
   Alpha=@n11
  )
  n23 = TextureSampleParameter2D(UVs=@n22) {group=SpinningStars, parameterName=SS_Texture_BC_01, sortPriority=10, texture=&T_BlacksRGB}
  n24 = ComponentMask(None=@n23.RGB) {b=true, desc=Base Texture, g=true, r=true}
  n25 = FunctionInput(Preview=Constant3Vector() {constant=<rgba 1 0 0 0>}) {bUsePreviewValueAsDefault=true, inputName=ForwardAxis}
  n26 = VectorParameter() {group=SpinningStars, parameterName=SS_Mask_WeaponOffset_ CE, sortPriority=180, defaultValue=<rgba 0 -1 0 1>}
  n27 = VectorParameter() {group=SpinningStars, parameterName=SS_Mask_VehicleOffset_ CE, sortPriority=181, defaultValue=<rgba -1 0 0 1>}
  n28 = VectorParameter() {group=SpinningStars, parameterName=SS_Mask_MechOffset_ CE, sortPriority=182, defaultValue=<rgba 0 0 -1 1>}
  n29 = VectorParameter() {group=SpinningStars, parameterName=SS_Mask_UV_Spread_XY CE, sortPriority=183, defaultValue=<rgba 1 1 1 0>}
  n30 = VectorParameter() {group=SpinningStars, parameterName=SS_Mask_UV_Spread_Mech_XY CE, sortPriority=184, defaultValue=<rgba 1 1 1 0>}
  n31 = LinearInterpolate(
   A=VectorParameter() {group=SpinningStars, parameterName=SS_Mask_UV_AnimatedWeapon_XY CE, sortPriority=185}.RGB
   B=VectorParameter() {group=SpinningStars, parameterName=SS_Mask_UV_AnimatedVehicle_XY CE, sortPriority=186}.RGB
   Alpha=@n9
  )
  n32 = VectorParameter() {group=SpinningStars, parameterName=SS_Mask_UV_AnimatedMech_XY CE, sortPriority=187}
  n33 = FunctionInput(Preview=VertexInterpolator(VS=PreSkinnedPosition()).PS) {bUsePreviewValueAsDefault=true, inputName=Local Position, sortPriority=11}
  n34 = TextureObjectParameter() {
   group=TopSideProjection
   parameterName=TS_Mask_01
   samplerType=SAMPLERTYPE_Masks
   sortPriority=20
   texture=&T_BlackMask
  }
  n35 = TextureSample(
   UVs=MaterialFunctionCall!(
    ForwardAxis=@n25
    Bounds=@n3
    Pre-Skinned Local Position=@n33
    isVehicle=@n9
    isMech=@n11
    Side Project Vehicle
    WeaponOffset=@n26.RGB
    WeaponFalloff=ScalarParameter() {defaultValue=0.5, group=SpinningStars, parameterName=SS_Mask_WeaponFalloff CE, sortPriority=180}
    VehicleOffset=@n27.RGB
    VehicleFalloff=ScalarParameter() {defaultValue=0.5, group=SpinningStars, parameterName=SS_Mask_VehicleFalloff CE, sortPriority=181}
    MechOffset=@n28.RGB
    MechFalloff=ScalarParameter() {group=SpinningStars, parameterName=SS_Mask_MechFalloff CE, sortPriority=182}
    Invert_Gradent
    UV_Spread_XY=LinearInterpolate(A=@n29.RGB, B=@n30.RGB, Alpha=@n11)
    UV_Animated_XY=LinearInterpolate(A=@n31, B=@n32.RGB, Alpha=@n11)
    Warp_UVs=ScalarParameter() {defaultValue=0.1, group=SpinningStars, parameterName=SS_Mask_WarpAmount, sortPriority=185}
    Rotate_Mech
    Rotate_Vehicle_Side
    Rotate_Vehicle
   ) {materialFunction=&/MF_Wrap_Gradent_Creation}.Uvs
   Tex=@n34
  )
  n36 = Time()
  n37 = ScalarParameter() {group=00-Reactive-GTFB, parameterName=00-GameTimeFromBP, sortPriority=1}
  n38 = Saturate(None=Divide(
    A=Subtract(A=@n36, B=@n37)
    B=ScalarParameter() {defaultValue=1, group=SpinningStars_Reactive, parameterName=SS_AnimationDuration, sortPriority=2}
   ))
  n39 = ScalarParameter() {defaultValue=0.5, group=SpinningStars_Reactive, parameterName=SS_Curve_Amount, sortPriority=3}
  n40 = ScalarParameter() {group=SpinningStars_Reactive, parameterName=SS_Reactive, sortPriority=-10}
  n41 = LinearInterpolate(A=OneMinus(None=Power(Base=@n38, Exp=@n39)) {desc=inverts curve from 0-1 to 1-0}, B=@n40, Alpha=@n40)
  n42 = Multiply(
   A=Clamp(
    None=ChannelMaskParameter(None=@n35.RGBA) {group=SpinningStars, parameterName=SS_Mask_Gradient_TextureChannel, sortPriority=120}.R
    Max=ScalarParameter() {defaultValue=1, group=SpinningStars, parameterName=SS_Mask_Gradent_Clamp Max, sortPriority=151}
   ) {maxDefault=0.99, minDefault=0.01}
   B=LinearInterpolate(
    A=ScalarParameter() {defaultValue=1, group=SpinningStars, parameterName=SS_Mask_Gradent_Opacity_CE, sortPriority=149}
    B=ScalarParameter() {defaultValue=0.8, group=SpinningStars_Reactive, parameterName=SS_Reactive_Mask_Gradent_Opacity_CE, sortPriority=6}
    Alpha=@n41
   )
  )
  n43 = VectorParameter() {group=TopSideProjection, parameterName=TS_Fill_alternate_Axis_Color_Mask_01, sortPriority=21, defaultValue=<rgba 0 0 0 1>}
  n44 = VectorParameter() {group=TopSideProjection_UVS, parameterName=TS-Rotation-Weapon-01---------------------, sortPriority=1, defaultValue=<rgba 0 0 -0.25 1>}
  n45 = VectorParameter() {group=TopSideProjection_UVS, parameterName=TS_Rotation_Vehicle_01, sortPriority=2, defaultValue=<rgba 0 0 0 1>}
  n46 = VectorParameter() {
   desc=must activate TS_Vehicle_Side_01 UVs
   group=TopSideProjection_UVS
   parameterName=TS_Rotation_Vehicle_Side_01
   sortPriority=2
   defaultValue=<rgba 0 0 0 1>
  }
  n47 = Multiply(A=@n8, B=ScalarParameter() {group=TopSideProjection_UVS, parameterName="TS_Vehicle_Side_01 UVs 0=Off 1=Active"})
  n48 = LinearInterpolate(A=@n45.RGB, B=@n46.RGB, Alpha=@n47)
  n49 = VectorParameter() {group=TopSideProjection_UVS, parameterName=TS_Rotation_Mech_01, sortPriority=3, defaultValue=<rgba 0.25 0.25 0 1>}
  n50 = FunctionInput(Preview=VertexInterpolator(VS=PreSkinnedNormal()).PS) {bUsePreviewValueAsDefault=true, inputName=Pre-Skinned Local Normal, sortPriority=10}
  n51 = VectorParameter() {group=TopSideProjection_UVS, parameterName=TS_Offset_Weapon_01---------------------, sortPriority=10, defaultValue=<rgba 0 0 0 1>}
  n52 = VectorParameter() {group=TopSideProjection_UVS, parameterName=TS_Offset_Vehicle_01, sortPriority=11, defaultValue=<rgba 0 0 0 1>}
  n53 = VectorParameter() {
   desc=must activate TS_Vehicle_Side_01 UVs
   group=TopSideProjection_UVS
   parameterName=TS_Offset_Vehicle_Side_01
   sortPriority=11
   defaultValue=<rgba 0 0 0 1>
  }
  n54 = LinearInterpolate(A=@n52.RGB, B=@n53.RGB, Alpha=@n47)
  n55 = VectorParameter() {group=TopSideProjection_UVS, parameterName=TS_Offset_Mech_01, sortPriority=12, defaultValue=<rgba 0 0 0 1>}
  n56 = Add(A=@n3, B=LinearInterpolate(A=LinearInterpolate(A=@n51.RGB, B=@n54, Alpha=@n9), B=@n55.RGB, Alpha=@n11))
  n57 = VectorParameter() {group=TopSideProjection_UVS, parameterName=TS_Panning_Weapon_01---------------------, sortPriority=40, defaultValue=<rgba 0 0 0 1>}
  n58 = VectorParameter() {group=TopSideProjection_UVS, parameterName=TS_Panning_Vehicle_01, sortPriority=41, defaultValue=<rgba 0 0 0 1>}
  n59 = VectorParameter() {
   desc=must activate TS_Vehicle_Side_01 UVs
   group=TopSideProjection_UVS
   parameterName=TS_Panning_Vehicle_Side_01
   sortPriority=41
   defaultValue=<rgba 0 0 0 1>
  }
  n60 = LinearInterpolate(A=@n58.RGB, B=@n59.RGB, Alpha=@n47)
  n61 = VectorParameter() {group=TopSideProjection_UVS, parameterName=TS_Panning_Mech_01, sortPriority=42, defaultValue=<rgba 0 0 0 1>}
  n62 = Multiply(A=Time(), B=LinearInterpolate(A=LinearInterpolate(A=@n57.RGB, B=@n60, Alpha=@n9), B=@n61.RGB, Alpha=@n11))
  n63 = VectorParameter() {group=TopSideProjection_UVS, parameterName="TS_Scale_Weapon_01 A=uniform", sortPriority=30, defaultValue=<rgba 1 1 1 1>}
  n64 = VectorParameter() {group=TopSideProjection_UVS, parameterName="TS_Scale_Vehicle_01 A=uniform", sortPriority=31, defaultValue=<rgba 1 1 1 1>}
  n65 = VectorParameter() {
   desc=must activate TS_Vehicle_Side_01 UVs
   group=TopSideProjection_UVS
   parameterName="TS_Scale_Vehicle_Side_01 A=uniform"
   sortPriority=31
   defaultValue=<rgba 1 1 1 1>
  }
  n66 = LinearInterpolate(
   A=Multiply(A=@n63.RGB, B=@n63.RGB)
   B=LinearInterpolate(A=Multiply(A=@n64.RGB, B=@n64.RGB), B=Multiply(A=@n65.RGB, B=@n65.RGB), Alpha=@n47)
   Alpha=@n9
  )
  n67 = VectorParameter() {group=TopSideProjection_UVS, parameterName="TS_Scale_Mech_01 A=uniform", defaultValue=<rgba 1 1 1 1>}
  n68 = Multiply(A=@n67.RGB, B=@n67.RGB)
  n69 = MaterialFunctionCall!(
   IsCharacter
   IsStatsB
   Texture=@n34
   Rotation_Axis=LinearInterpolate(A=LinearInterpolate(A=@n44.RGB, B=@n48, Alpha=@n9), B=@n49.RGB, Alpha=@n11)
   Caculated_Bounds=Multiply(A=Add(A=@n56, B=@n62), B=LinearInterpolate(A=@n66, B=@n68, Alpha=@n11))
   Pre_Skinned_Normals=@n50
   Blend_Contrast=LinearInterpolate(
    A=ScalarParameter() {defaultValue=1, group=TopSideProjection, parameterName=TS_Blend_Contrast, sortPriority=-1}
    B=ScalarParameter() {defaultValue=2, group=TopSideProjection, parameterName=TS_Blend_Contrast_Mech, sortPriority=-1}
    Alpha=@n11
   )
   Fill_alternate_Axis_Opacity=ScalarParameter() {group=TopSideProjection, parameterName=TS_Fill_alternate_Axis_Opacity 01, sortPriority=2}
   "Fill_alternate_Axis_Color_A=Roughness"=AppendVector(A=@n43.RGB, B=@n43.RGB)
   Is_Vehicle=@n9
   Is_Mech=@n11
   Enable_Fill_Z_Weapon
   Fill_Z_Weapon_opacity
   Fill_Z_Weapon_Color
  ) {materialFunction=&/MF_TopsSideProjection_Base}
  n70 = Saturate(None=MaterialFunctionCall(
    Transition Phase=Saturate(None=@n42)
    Height Texture=ChannelMaskParameter(None=@n69.Combined) {group=SpinningStars, parameterName=TS_Mask_HeightTexture, sortPriority=119}.R
    Contrast=ScalarParameter() {defaultValue=1, group=SpinningStars, parameterName=SS_Mask_HeightLerp_Contrast, sortPriority=145}
   ) {materialFunction=&HeightLerp}.Alpha)
  n71 = TextureSampleParameter2D(UVs=@n22) {
   group=SpinningStars
   parameterName=SS_HeightLerp_Texture
   sortPriority=118
   texture=/Game/Weapons/WeaponSkins/Wraps/Textures/SpinningStar/T_Wrap_SpinningStar_Grad
  }
  n72 = Add(
   A=ScalarParameter() {group=SpinningStars_Reactive, parameterName=SS_HueShift_BaseColor, sortPriority=4}
   B=Multiply(A=Time(), B=ScalarParameter() {group=SpinningStars_Reactive, parameterName=SS_HueShift_BaseColor_Animated, sortPriority=5})
  )
  n73 = Multiply(
   A=@n70
   B=LinearInterpolate(
    A=@n71.RGB
    B=MaterialFunctionCall(Hue Shift Percentage=@n72, Texture=@n71.RGB) {materialFunction=&HueShift}.Result
    Alpha=@n41
   )
  )
  n74 = ScalarParameter() {defaultValue=10, group=SpinningStars, parameterName=SS_SpeculartyFresnelExponent, sortPriority=12}
  n75 = Multiply(A=@n73, B=ScalarParameter() {defaultValue=1, group=SpinningStars, parameterName=SS_MainStar_EmissiveIntensity, sortPriority=118}) {constB=0.5}
  n76 = ScalarParameter() {defaultValue=1, group=SpinningStars, parameterName=CS_BaseStarEmissiveIntensity, sortPriority=198}
  n77 = VectorParameter() {group=SpinningStars, parameterName=TS_SecondaryStars_Color, sortPriority=199, defaultValue=<rgba 3 3 3 1>}
  n78 = ScalarParameter() {
   defaultValue=1
   desc="emissive color = 1, Emissive using base texture = 0"
   group=SpinningStars
   parameterName=Use Texture for Emissive
   sortPriority=198
  }
  n79 = VectorParameter() {group=SpinningStars, parameterName=SS_SecondaryStars_WeaponOffset_ CE, sortPriority=280, defaultValue=<rgba 0 -1 0 1>}
  n80 = VectorParameter() {group=SpinningStars, parameterName=SS_SecondaryStars_VehicleOffset_ CE, sortPriority=281, defaultValue=<rgba -1 0 0 1>}
  n81 = VectorParameter() {group=SpinningStars, parameterName=SS_SecondaryStars_MechOffset_ CE, sortPriority=282, defaultValue=<rgba 0 0 -1 1>}
  n82 = VectorParameter() {group=SpinningStars, parameterName=SS_SecondaryStars_UV_Spread_XY CE, sortPriority=283, defaultValue=<rgba 1 1 1 0>}
  n83 = VectorParameter() {group=SpinningStars, parameterName=SS_SecondaryStars_UV_Spread_Mech_XY CE, sortPriority=284, defaultValue=<rgba 1 1 1 0>}
  n84 = LinearInterpolate(
   A=VectorParameter() {group=SpinningStars, parameterName=SS_SecondaryStars_UV_AnimatedWeapon_XY CE, sortPriority=285}.RGB
   B=VectorParameter() {group=SpinningStars, parameterName=SS_SecondaryStars_UV_AnimatedVehicle_XY CE, sortPriority=286}.RGB
   Alpha=@n9
  )
  n85 = VectorParameter() {group=SpinningStars, parameterName=SS_SecondaryStars_UV_AnimatedMech_XY CE, sortPriority=287}
  n86 = TextureObjectParameter() {
   group=TopSideProjection
   parameterName=TS_Mask_02
   samplerType=SAMPLERTYPE_Masks
   sortPriority=62
   texture=&T_BlackMask
  }
  n87 = VectorParameter() {group=TopSideProjection, parameterName=TS_Fill_alternate_Axis_Color_Mask 02, sortPriority=67, defaultValue=<rgba 0 0 0 1>}
  n88 = VectorParameter() {group=TopSideProjection_UVS, parameterName=TS-Rotation-Weapon-02---------------------, sortPriority=150, defaultValue=<rgba 0 0 -0.25 1>}
  n89 = VectorParameter() {group=TopSideProjection_UVS, parameterName=TS_Rotation_Vehicle_02, sortPriority=151, defaultValue=<rgba 0 0 0 1>}
  n90 = VectorParameter() {
   desc=must activate TS_Vehicle_Side_02 UVs
   group=TopSideProjection_UVS
   parameterName=TS_Rotation_Vehicle_Side_02
   sortPriority=151
   defaultValue=<rgba 0 0 0 1>
  }
  n91 = Multiply(A=@n8, B=ScalarParameter() {group=TopSideProjection_UVS, parameterName="TS_Vehicle_Side_02 UVs  0=Off 1=Active", sortPriority=149})
  n92 = LinearInterpolate(A=@n89.RGB, B=@n90.RGB, Alpha=@n91)
  n93 = VectorParameter() {group=TopSideProjection_UVS, parameterName=TS_Rotation_Mech_02, sortPriority=153, defaultValue=<rgba 0.25 0.25 0 1>}
  n94 = VectorParameter() {group=TopSideProjection_UVS, parameterName=TS_Offset_Weapon_02---------------------, sortPriority=160, defaultValue=<rgba 0 0 0 1>}
  n95 = VectorParameter() {group=TopSideProjection_UVS, parameterName=TS_Offset_Vehicle_02, sortPriority=161, defaultValue=<rgba 0 0 0 1>}
  n96 = VectorParameter() {
   desc=must activate TS_Vehicle_Side_02 UVs
   group=TopSideProjection_UVS
   parameterName=TS_Offset_Vehicle_Side_02
   sortPriority=161
   defaultValue=<rgba 0 0 0 1>
  }
  n97 = LinearInterpolate(A=@n94.RGB, B=LinearInterpolate(A=@n95.RGB, B=@n96.RGB, Alpha=@n91), Alpha=@n9)
  n98 = VectorParameter() {group=TopSideProjection_UVS, parameterName=TS_Offset_Mech_02, sortPriority=162, defaultValue=<rgba 0 0 0 1>}
  n99 = VectorParameter() {group=TopSideProjection_UVS, parameterName=TS_Panning_Weapon_02---------------------, sortPriority=180, defaultValue=<rgba 0 0 0 1>}
  n100 = VectorParameter() {group=TopSideProjection_UVS, parameterName=TS_Panning_Vehicle_02, sortPriority=181, defaultValue=<rgba 0 0 0 1>}
  n101 = VectorParameter() {
   desc=must activate TS_Vehicle_Side_02 UVs
   group=TopSideProjection_UVS
   parameterName=TS_Panning_Vehicle_Side_02
   sortPriority=181
   defaultValue=<rgba 0 0 0 1>
  }
  n102 = LinearInterpolate(A=@n99.RGB, B=LinearInterpolate(A=@n100.RGB, B=@n101.RGB, Alpha=@n91), Alpha=@n9)
  n103 = VectorParameter() {group=TopSideProjection_UVS, parameterName=TS_Panning_Mech_02, sortPriority=182, defaultValue=<rgba 0 0 0 1>}
  n104 = Add(
   A=Add(A=@n3, B=LinearInterpolate(A=@n97, B=@n98.RGB, Alpha=@n11))
   B=Multiply(A=Time(), B=LinearInterpolate(A=@n102, B=@n103.RGB, Alpha=@n11))
  )
  n105 = VectorParameter() {group=TopSideProjection_UVS, parameterName="TS_Scale_Weapon_02 A=uniform", sortPriority=170, defaultValue=<rgba 1 1 1 1>}
  n106 = VectorParameter() {group=TopSideProjection_UVS, parameterName="TS_Scale_Vehicle_02 A=uniform", sortPriority=171, defaultValue=<rgba 1 1 1 1>}
  n107 = Multiply(A=@n106.RGB, B=@n106.RGB)
  n108 = VectorParameter() {
   desc=must activate TS_Vehicle_Side_02 UVs
   group=TopSideProjection_UVS
   parameterName="TS_Scale_Vehicle_Side_02 A=uniform"
   sortPriority=171
   defaultValue=<rgba 1 1 1 1>
  }
  n109 = Multiply(A=@n108.RGB, B=@n108.RGB)
  n110 = VectorParameter() {group=TopSideProjection_UVS, parameterName="TS_Scale_Mech_02 A=uniform", sortPriority=172, defaultValue=<rgba 1 1 1 1>}
  n111 = LinearInterpolate(
   A=LinearInterpolate(A=Multiply(A=@n105.RGB, B=@n105.RGB), B=LinearInterpolate(A=@n107, B=@n109, Alpha=@n91), Alpha=@n9)
   B=Multiply(A=@n110.RGB, B=@n110.RGB)
   Alpha=@n11
  )
  n112 = TextureCoordinate()
  n113 = LinearInterpolate(
   A=ScalarParameter() {defaultValue=1, group=TopSideProjection_UVS, parameterName=TS-WarpUVS-02-TileRate Weapon, sortPriority=203}
   B=ScalarParameter() {defaultValue=1, group=TopSideProjection_UVS, parameterName=TS-WarpUVS-02-TileRate Vehicle, sortPriority=204}
   Alpha=@n9
  )
  n114 = Time() {period=1}
  n115 = VectorParameter() {group=TopSideProjection_UVS, parameterName=TS-WarpUVS-02-Anitmated Weapon, sortPriority=207, defaultValue=<rgba 0 0 0 1>}
  n116 = VectorParameter() {group=TopSideProjection_UVS, parameterName=TS-WarpUVS-02-Anitmated Vehicle, sortPriority=208, defaultValue=<rgba 0 0 0 1>}
  n117 = ComponentMask(None=LinearInterpolate(A=@n115.RGB, B=@n116.RGB, Alpha=@n9)) {g=true, r=true}
  n118 = TextureSampleParameter2D(UVs=Add(A=Multiply(A=@n112, B=@n113), B=Multiply(A=@n114, B=@n117))) {group=TopSideProjection_UVS, parameterName=TS-WarpUVS-02-Texture, sortPriority=202, texture=&T_BlacksRGB}
  n119 = ChannelMaskParameter(None=@n118.RGB) {group=TopSideProjection_UVS, parameterName=TS-WarpUVS-02_Channel, sortPriority=201}
  n120 = LinearInterpolate(
   A=ScalarParameter() {group=TopSideProjection_UVS, parameterName=TS-WarpUVS-02-Amount Weapon, sortPriority=205}
   B=ScalarParameter() {group=TopSideProjection_UVS, parameterName=TS-WarpUVS-02-Amount_Vehicle, sortPriority=206}
   Alpha=@n9
  )
  n121 = MaterialFunctionCall(
   Texture=@n86
   Rotation_Axis=LinearInterpolate(A=LinearInterpolate(A=@n88.RGB, B=@n92, Alpha=@n9), B=@n93.RGB, Alpha=@n11)
   Caculated_Bounds=Add(A=Multiply(A=@n104, B=@n111), B=Multiply(A=@n119.R, B=@n120))
   Pre_Skinned_Normals=@n50
   Blend_Contrast=LinearInterpolate(
    A=ScalarParameter() {defaultValue=1, group=TopSideProjection, parameterName=TS_Blend_Contrast, sortPriority=-1}
    B=ScalarParameter() {defaultValue=2, group=TopSideProjection, parameterName=TS_Blend_Contrast_Mech, sortPriority=-1}
    Alpha=@n11
   )
   Fill_alternate_Axis_Opacity=ScalarParameter() {group=TopSideProjection, parameterName=TS_Fill_alternate_Axis_Opacity 02, sortPriority=51}
   "Fill_alternate_Axis_Color_A=Roughness"=AppendVector(A=@n87.RGB, B=@n87.RGB)
   Is_Vehicle=@n9
   Is_Mech=@n11
  ) {materialFunction=&MF_TopsSideProjection_Base}
  n122 = ChannelMaskParameter(None=@n121.Combined) {group=SpinningStars, parameterName=TS_SecondaryStars_WarpTexture, sortPriority=200}
  n123 = ScalarParameter() {defaultValue=0.1, group=SpinningStars, parameterName=SS_SecondaryStars_WarpAmount, sortPriority=288}
  n124 = TextureSample(
   UVs=MaterialFunctionCall(
    ForwardAxis=@n25
    Bounds=@n3
    Pre-Skinned Local Position=@n33
    isVehicle=@n9
    isMech=@n11
    WeaponOffset=@n79.RGB
    WeaponFalloff=ScalarParameter() {defaultValue=0.5, group=SpinningStars, parameterName=SS_SecondaryStars_WeaponFalloff CE, sortPriority=280}
    VehicleOffset=@n80.RGB
    VehicleFalloff=ScalarParameter() {defaultValue=0.5, group=SpinningStars, parameterName=SS_SecondaryStars_VehicleFalloff CE, sortPriority=281}
    MechOffset=@n81.RGB
    MechFalloff=ScalarParameter() {group=SpinningStars, parameterName=SS_SecondaryStars_MechFalloff CE, sortPriority=282}
    UV_Spread_XY=LinearInterpolate(A=@n82.RGB, B=@n83.RGB, Alpha=@n11)
    UV_Animated_XY=LinearInterpolate(A=@n84, B=@n85.RGB, Alpha=@n11)
    Warp_UVs=Multiply(A=@n122.R, B=@n123)
   ) {materialFunction=&MF_Wrap_Gradent_Creation}.Uvs
   Tex=@n86
  )
  n125 = Multiply(
   A=Clamp(
    None=ChannelMaskParameter(None=@n124.RGBA) {group=SpinningStars, parameterName=SS_SecondaryStars_Gradient_TextureChannel, sortPriority=220}.R
    Max=ScalarParameter() {defaultValue=1, group=SpinningStars, parameterName=SS_SecondaryStarsGradient_Clamp Max, sortPriority=251}
   ) {maxDefault=0.99, minDefault=0.01}
   B=ScalarParameter() {defaultValue=1, group=SpinningStars, parameterName=SS_SecondaryStarsGradient_Opacity_CE, sortPriority=249}
  )
  n126 = Saturate(None=MaterialFunctionCall(
    Transition Phase=Saturate(None=@n125)
    Height Texture=ChannelMaskParameter(None=@n121.Combined) {group=SpinningStars, parameterName=TS_SecondaryStars_HeightTexture, sortPriority=200}.R
    Contrast=ScalarParameter() {defaultValue=1, group=SpinningStars, parameterName=SS_SecondaryStarsHeightLerp_Contrast, sortPriority=245}
   ) {materialFunction=&HeightLerp}.Alpha)
  n127 = Multiply(A=OneMinus(None=@n70), B=@n126)
  n128 = Add(
   A=LinearInterpolate(A=Multiply(A=@n73, B=@n76), B=Multiply(A=@n70, B=@n77.RGB), Alpha=@n78)
   B=LinearInterpolate(A=Constant(), B=Multiply(A=@n77.RGB, B=@n127), Alpha=@n126)
  )
  n129 = SetMaterialAttributes!(
   MaterialAttributes=@n2.MaterialAttributes
   BaseColor=LinearInterpolate(A=@n24, B=LinearInterpolate(A=@n24, B=@n73, Alpha=@n70), Alpha=@n71.A)
   Roughness=LinearInterpolate(
    A=ScalarParameter() {group=SpinningStars, parameterName=SS_Roughness_Min, sortPriority=11}
    B=ScalarParameter() {defaultValue=1, group=SpinningStars, parameterName=SS_Roughness_Max, sortPriority=11}
    Alpha=@n23.A
   )
   Specular=Add(A=ScalarParameter() {defaultValue=0.5, group=SpinningStars, parameterName=SS_Specularty, sortPriority=12}, B=Fresnel(ExponentIn=@n74))
   EmissiveColor=LinearInterpolate(B=LinearInterpolate(A=@n75, B=@n128, Alpha=@n41), Alpha=@n71.A)
   Metallic=Constant()
  ) {attributeSetTypes [
    #2
    #3
    #4
    #5
    #6
   ]}
  n130 = FunctionInput() {
   bUsePreviewValueAsDefault=true
   description=primary
   inputName="Customization Mask 2 (_CM B Top Right)"
   inputType=FunctionInput_Scalar
   sortPriority=25
  }
  n131 = FunctionInput() {
   bUsePreviewValueAsDefault=true
   description=secondary
   inputName="Customization Mask (_CM B Top left)"
   inputType=FunctionInput_Scalar
   sortPriority=24
  }
  n132 = Saturate(None=Add(A=Add(A=@n130, B=@n131), B=ScalarParameter() {group=SpinningStars, parameterName=SS_Remove_Customization Mask, sortPriority=-10}))
  n133 = Saturate(None=Subtract(A=NamedRerouteUsage(), B=NamedRerouteUsage()))
  n134 = StaticSwitchParameter(True=BlendMaterialAttributes(A=@n2.MaterialAttributes, B=@n129, Alpha=LinearInterpolate(A=@n132, B=@n131, Alpha=@n133)), False=@n1) {
   desc=Side Projection Base - Topside Height Lerp - Masked by Side - Reactive 2nd Height Lerp and Hue Shift Glow
   group=Customization_Wrap
   parameterName=Wrap_SpinningStars
  }
  FunctionOutput(None=@n134)
  NamedRerouteDeclaration(None=FunctionInput() {bUsePreviewValueAsDefault=true, inputName=IsCharacter, inputType=FunctionInput_Scalar}) {name=IsCharacter, nodeColor=<rgba 0.9859 1 0 1>}
  NamedRerouteDeclaration(None=FunctionInput() {bUsePreviewValueAsDefault=true, inputName=isCharAccent, inputType=FunctionInput_Scalar, sortPriority=78}) {name=isCharAccent, nodeColor=<rgba 0 0.9253 1 1>}
 ]}
