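# TNTina weapon wrap material function that replaces Base Color inside two customization masks when its static switch is on.
# Normalized pre-skinned local position, adjusted by vehicle and mech inputs, drives a two-color background gradient combined with a GMap input, plus a "bang" color gradient.
# Blueprint-fed timestamps animate the bang spread and slide a box-masked flakes texture, whose tiled UVs take a hashed seed offset.
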MaterialFunction /Game/Weapons/WeaponSkins/Wraps/MaterialFunctions/MF_TNTina_Wrap
graph {nodes [
  n1 = GetMaterialAttributes(None=FunctionInput(Preview=MakeMaterialAttributes!(
     BaseColor
     Metallic
     Specular
     Roughness
     Anisotropy
     EmissiveColor
     Opacity
     OpacityMask
     Normal
     Tangent
     WorldPositionOffset
     SubsurfaceColor
     ClearCoat
     ClearCoatRoughness
     AmbientOcclusion
     Refraction
     CustomizedUVs_0
     CustomizedUVs_1
     CustomizedUVs_2
     CustomizedUVs_3
     CustomizedUVs_4
     CustomizedUVs_5
     CustomizedUVs_6
     CustomizedUVs_7
     PixelDepthOffset
     ShadingModel
     Displacement
    )) {inputName=Material, inputType=FunctionInput_MaterialAttributes, sortPriority=-200}) {attributeGetTypes [#1]}
  n2 = FunctionInput(Preview=VertexInterpolator(VS=PreSkinnedPosition()).PS) {bUsePreviewValueAsDefault=true, inputName=Pre-Skinned Local Position, sortPriority=41}
  n3 = MaterialFunctionCall(Float3=@n2) {materialFunction=&BreakOutFloat3Components}
  n4 = FunctionInput(Preview=Constant()) {inputName=isVehicle, inputType=FunctionInput_Scalar}
  n5 = FunctionInput(Preview=Constant()) {bUsePreviewValueAsDefault=true, inputName=isMech, inputType=FunctionInput_Scalar}
  n6 = LinearInterpolate(
   A=LinearInterpolate(
    A=@n2
    B=MaterialFunctionCall(X=@n3.B, Y=@n3.B, Z=@n3.B) {materialFunction=&MakeFloat3}.Result
    Alpha=@n4
   )
   B=MaterialFunctionCall(X=@n3.R, Y=@n3.R, Z=@n3.R) {materialFunction=&MakeFloat3}.Result
   Alpha=@n5
  )
  n7 = PreSkinnedLocalBounds()
  n8 = MaterialFunctionCall(Float3=@n7.Min) {materialFunction=&BreakOutFloat3Components}
  n9 = LinearInterpolate(
   A=@n7.Min
   B=MaterialFunctionCall(X=@n8.B, Y=@n8.B, Z=@n8.B) {materialFunction=&MakeFloat3}.Result
   Alpha=@n4
  )
  n10 = MaterialFunctionCall(X=@n8.R, Y=@n8.R, Z=@n8.R) {materialFunction=&MakeFloat3}
  n11 = Multiply(
   A=Subtract(A=@n6, B=LinearInterpolate(A=@n9, B=@n10.Result, Alpha=@n5))
   B=MaterialFunctionCall() {materialFunction=/Engine/Functions/Engine_MaterialFunctions02/WorldPositionOffset/ObjectScale}.Scale XYZ
  )
  n12 = FunctionInput(Preview=VertexInterpolator(VS=@n7.Half Extents).PS) {bUsePreviewValueAsDefault=true, inputName=Local Bounds}
  n13 = MaterialFunctionCall(Float3=@n12) {materialFunction=&BreakOutFloat3Components}
  n14 = MaterialFunctionCall(X=@n13.B, Y=@n13.B, Z=@n13.B) {materialFunction=&MakeFloat3}
  n15 = Multiply(A=LinearInterpolate(
    A=LinearInterpolate(A=@n12, B=@n14.Result, Alpha=@n4)
    B=MaterialFunctionCall(X=@n13.R, Y=@n13.R, Z=@n13.R) {materialFunction=&MakeFloat3}.Result
    Alpha=@n5
   )) {constB=2}
  n16 = Max(A=ComponentMask(None=@n15) {g=true}, B=ComponentMask(None=@n15) {b=true})
  n17 = AppendVector(A=AppendVector(A=ComponentMask(None=@n15) {r=true}, B=@n16), B=@n16)
  n18 = Add(
   A=ComponentMask(None=Divide(A=@n11, B=@n17)) {b=true, g=true}
   B=ComponentMask(None=VectorParameter() {group=TN_Tina, parameterName=TNT_BoundsOffset_XY, defaultValue=<rgba 0 0 0 1>}.RGB) {g=true, r=true}
  ) {constB=0.25}
  n19 = ComponentMask(None=@n18) {r=true}
  n20 = Subtract(A=@n19, B=ScalarParameter() {group=TN_Tina, parameterName=TNT_GradientOffset})
  n21 = Constant()
  n22 = Multiply(
   A=Subtract(
    A=FunctionInput(Preview=@n21) {bUsePreviewValueAsDefault=true, inputName=GMap, inputType=FunctionInput_Scalar}
    B=ScalarParameter() {group=TN_Tina, parameterName=TNT_GMapSpread}
   )
   B=ScalarParameter() {defaultValue=1, group=TN_Tina, parameterName=TNT_GmapStrength}
  )
  n23 = LinearInterpolate(
   A=VectorParameter() {group=TN_Tina, parameterName=TNT_BG_Color_01, defaultValue=<rgba 0.8308 0.0513 0.0844 1>}.RGB
   B=VectorParameter() {group=TN_Tina, parameterName=TNT_BG_Color_02, defaultValue=<rgba 0.0052 0.0003 0.0005 1>}.RGB
   Alpha=Max(A=Clamp(None=@n20), B=Clamp(None=@n22))
  )
  n24 = VectorParameter() {group=TN_Tina, parameterName=TNT_Bang_Color, defaultValue=<rgba 0.8308 0.4968 0 1>}
  n25 = ScalarParameter() {group=TN_Tina, parameterName=TNT_BangGradient}
  n26 = Multiply(A=@n25, B=ScalarParameter() {defaultValue=2, parameterName=TNT_BangSpread}) {constB=2}
  n27 = Subtract(A=Time(), B=ScalarParameter() {defaultValue=-999999995904, group=TN_Tina, parameterName=TNT_GameTimeFromBP})
  n28 = ScalarParameter() {defaultValue=1, group=TN_Tina, parameterName=TNT_AnimSpeed}
  n29 = OneMinus(None=Saturate(None=Divide(A=@n27, B=@n28)))
  n30 = Clamp(None=Subtract(A=@n19, B=LinearInterpolate(A=@n25, B=@n26, Alpha=@n29)))
  n31 = ScalarParameter() {defaultValue=1, group=TN_Tina, parameterName=TNT_FlakesTiling}
  n32 = MaterialFunctionCall(Vec3=VectorParameter() {parameterName=TNT_Vec3SeedFromBP, defaultValue=<rgba 0 0 0 1>}.RGB) {materialFunction=&MF_Hash}
  n33 = MaterialFunctionCall(Float2=Add(A=Multiply(A=@n18, B=@n31), B=ComponentMask(None=@n32.Vec3) {g=true, r=true})) {materialFunction=&BreakOutFloat2Components}
  n34 = Time()
  n35 = ScalarParameter() {defaultValue=-999999995904, group=TN_Tina, parameterName=TNT_GameTimeFromBP2}
  n36 = Saturate(None=Divide(A=Subtract(A=@n34, B=@n35), B=ScalarParameter() {defaultValue=1, group=TN_Tina, parameterName=TNT_FlakesAnimSpeed}))
  n37 = LinearInterpolate(
   A=Multiply(A=@n36, B=ScalarParameter() {defaultValue=2, group=TN_Tina, parameterName=TNT_FlakesMove_Amt})
   B=ScalarParameter() {group=TN_Tina, parameterName=DebugTime}
   Alpha=ScalarParameter() {group=TN_Tina, parameterName=TNT_DebugFlakeMovement}
  )
  n38 = TextureSampleParameter2D(UVs=MaterialFunctionCall(X=Add(A=@n33.R, B=@n37), Y=OneMinus(None=@n33.G)) {materialFunction=&MakeFloat2}.Result) {group=TN_Tina, parameterName=TNT_FlakesTexture, texture=/Game/Weapons/WeaponSkins/Wraps/Textures/TNTina/T_TNTina_Flakes}
  n39 = ComponentMask(None=VectorParameter() {group=TN_Tina, parameterName=TNT_FLakeMaskOffset, defaultValue=<rgba 0 0 0 1>}.RGB) {g=true, r=true}
  n40 = ScalarParameter() {defaultValue=1, group=TN_Tina, parameterName=TNT_Mask_X_Scale}
  n41 = ScalarParameter() {defaultValue=1, group=TN_Tina, parameterName=TNT_Mask_Y_Scale}
  n42 = MaterialFunctionCall(Float2=Multiply(A=Add(A=@n18, B=@n39), B=AppendVector(A=@n40, B=@n41))) {materialFunction=&BreakOutFloat2Components}
  n43 = Subtract(
   A=MaterialFunctionCall(X=Add(A=@n42.R, B=@n37), Y=OneMinus(None=@n42.G)) {materialFunction=&MakeFloat2}.Result
   B=Constant2Vector() {g=0.5}
  )
  n44 = Desaturation(None=Round(None=Abs(None=@n43)), Fraction=Constant() {r=1})
  n45 = OneMinus(None=Clamp(None=Ceil(None=@n44)))
  n46 = LinearInterpolate(B=@n38.RGB, Alpha=@n45)
  n47 = LinearInterpolate(B=@n38.A, Alpha=@n45)
  n48 = FunctionInput() {bUsePreviewValueAsDefault=true, inputName="Customization Mask 1 (_CM B Top Left)", inputType=FunctionInput_Scalar, sortPriority=25}
  n49 = FunctionInput() {bUsePreviewValueAsDefault=true, inputName="Customization Mask 2 (_CM B Top Right)", inputType=FunctionInput_Scalar, sortPriority=26}
  n50 = LinearInterpolate(
   A=@n1.BaseColor
   B=LinearInterpolate(A=LinearInterpolate(A=@n23, B=@n24.RGB, Alpha=@n30), B=Multiply(A=@n46), Alpha=Multiply(A=@n47))
   Alpha=Saturate(None=Add(A=@n48, B=@n49))
  )
  FunctionOutput(None=StaticSwitchParameter(
    True=SetMaterialAttributes!(MaterialAttributes=@n1.MaterialAttributes, BaseColor=@n50) {attributeSetTypes [#1]}
    False=@n1.MaterialAttributes
   ) {group=Customization_Wrap, parameterName=TNTina_Wrap})
  Multiply(A=@n46, B=@n45)
  Multiply(A=@n47, B=@n45)
  OneMinus(None=Floor(None=@n36))
  n51 = FunctionInput(Preview=VertexInterpolator(VS=PreSkinnedNormal()).PS) {bUsePreviewValueAsDefault=true, inputName=Pre-Skinned Local Normal, sortPriority=42}
  n52 = MaterialFunctionCall(Float3=@n51) {materialFunction=&BreakOutFloat3Components}
  n53 = LinearInterpolate(
   A=@n51
   B=MaterialFunctionCall(X=@n52.B, Y=@n52.B, Z=@n52.B) {materialFunction=&MakeFloat3}.Result
   Alpha=@n4
  )
  n54 = MaterialFunctionCall(X=@n52.R, Y=@n52.R, Z=@n52.R) {materialFunction=&MakeFloat3}
  Reroute(None=MaterialFunctionCall!(
    WorldPosition=@n6
    World Space Normal=LinearInterpolate(A=@n53, B=@n54.Result, Alpha=@n5)
    UVSize_Front
    UVSize_Side
    UVSize_Top
    ProjectionTransitionContrast=Constant() {r=0.5}
   ) {materialFunction=/Game/Packages/Fortress_SharedMaterials/Base_Material_Functions/WorldAlignedTexture_UVsMasks}.Mask_Side)
 ]}
