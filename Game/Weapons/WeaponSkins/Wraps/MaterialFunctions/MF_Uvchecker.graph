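# MF_Uvchecker is a debug material function for weapon wraps that, behind a static switch, paints a UV checker over the input material's customization-mask areas.
# Scalar parameters blend base color between a top/side bounds projection, color-tinted Back, Middle and Front safe zones, and a virtual-plane projection.
# IsStates inputs adjust projection for object types such as vehicle and mech.
# Several branches, including the skew side projection, never reach the output.
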MaterialFunction /Game/Weapons/WeaponSkins/Wraps/MaterialFunctions/MF_Uvchecker
graph {nodes [
  n1 = Constant() {r=0.05}
  Add(A=@n1)
  n2 = StaticBool() {value=true}
  n3 = MaterialFunctionCall(Float4=FunctionInput(Preview=Constant4Vector().RGBA) {bUsePreviewValueAsDefault=true, inputName=IsStates, inputType=FunctionInput_Vector4, sortPriority=4}) {materialFunction=&BreakOutFloat4Components}
  n4 = MaterialFunctionCall(R=@n3.R, G=@n3.R, B=@n3.R) {materialFunction=&AppendMany}
  n5 = MaterialFunctionCall(Color=@n4.RGB) {materialFunction=&SplitComponents}
  n6 = Constant()
  n7 = FunctionInput(Preview=Constant4Vector() {constant=<rgba 0.3 0 0 0.25>}.RGBA) {bUsePreviewValueAsDefault=true, inputName=Object-SafeZone-Back, inputType=FunctionInput_Vector4, sortPriority=100}
  n8 = FunctionInput(Preview=VertexInterpolator(VS=PreSkinnedNormal()).PS) {bUsePreviewValueAsDefault=true, inputName=Pre-Skinned Local Normal, sortPriority=10}
  n9 = VertexInterpolator(VS=PreSkinnedPosition())
  n10 = PreSkinnedLocalBounds()
  n11 = FunctionInput(Preview=Divide(A=Subtract(A=@n9.PS, B=@n10.Min), B=Subtract(A=@n10.Max, B=@n10.Max))) {bUsePreviewValueAsDefault=true, inputName=Calcated_Bounds, sortPriority=11}
  n12 = MaterialFunctionCall!(
   For_Safe_Zones=@n2
   Normal_Map
   Caculated_Bounds=@n11
   Pre_Skinned_Normals=@n8
   Is Vehicle=@n5.R
   Is VehicleSide=@n5.R
   Is Mech=@n5.R
   Vehicle_Rotation
   UV_Rotation Bypass=@n6
   Mech_Rotation
   Uvs_Offset=ComponentMask(None=@n7) {b=true, g=true, r=true}
   Offset-after rotation
   Uvs_TileRate=ComponentMask(None=@n7) {a=true}
  ) {materialFunction=&/MF_Projection_Side_Normal}
  n13 = TextureObjectParameter() {group=0_UvCheker, parameterName=O_UvChecker_text, texture=/Game/Weapons/WeaponSkins/Wraps/Textures/UVCheaker}
  n14 = NamedRerouteDeclaration(None=@n13) {name=UvChecker_text, nodeColor=<rgba 1 0 0.6094 1>}
  n15 = TextureSample(UVs=@n12.UVs, Tex=@n14)
  n16 = Multiply(A=@n15.RGB, B=@n12.Safe_Zone_Mask)
  n17 = Constant3Vector() {constant=<rgba 0 1 1 0>}
  n18 = FunctionInput(Preview=Constant4Vector() {constant=<rgba 0 0 0 0.25>}.RGBA) {bUsePreviewValueAsDefault=true, inputName=Object-SafeZone-Middle, inputType=FunctionInput_Vector4, sortPriority=101}
  n19 = MaterialFunctionCall(
   For_Safe_Zones=@n2
   Caculated_Bounds=@n11
   Pre_Skinned_Normals=@n8
   Is Vehicle=@n5.R
   Is VehicleSide=@n5.R
   Is Mech=@n5.R
   UV_Rotation Bypass=@n6
   Uvs_Offset=ComponentMask(None=@n18) {b=true, g=true, r=true}
   Uvs_TileRate=ComponentMask(None=@n18) {a=true}
  ) {materialFunction=&MF_Projection_Side_Normal}
  n20 = Multiply(A=TextureSample(UVs=@n19.UVs, Tex=@n14).RGB, B=@n19.Safe_Zone_Mask)
  n21 = Constant3Vector() {constant=<rgba 1 1 0 0>}
  n22 = FunctionInput(Preview=Constant4Vector() {constant=<rgba -0.3 0 0 0.25>}.RGBA) {bUsePreviewValueAsDefault=true, inputName=Object-SafeZone-Front, inputType=FunctionInput_Vector4, sortPriority=102}
  n23 = MaterialFunctionCall(
   For_Safe_Zones=@n2
   Caculated_Bounds=@n11
   Pre_Skinned_Normals=@n8
   Is Vehicle=@n5.R
   Is VehicleSide=@n5.R
   Is Mech=@n5.R
   UV_Rotation Bypass=@n6
   Uvs_Offset=ComponentMask(None=@n22) {b=true, g=true, r=true}
   Uvs_TileRate=ComponentMask(None=@n22) {a=true}
  ) {materialFunction=&MF_Projection_Side_Normal}
  n24 = TextureSample(UVs=@n23.UVs, Tex=@n14)
  n25 = Add(
   A=Add(A=Multiply(A=@n16, B=@n17), B=Multiply(A=@n20, B=@n21))
   B=Multiply(A=Multiply(A=@n24.RGB, B=@n23.Safe_Zone_Mask), B=Constant3Vector() {constant=<rgba 1 0 1 0>})
  )
  n26 = MaterialFunctionCall(Color=@n4.RGB) {materialFunction=&SplitComponents}
  n27 = LinearInterpolate(A=LinearInterpolate(B=LinearInterpolate(Alpha=@n26.G) {constA=1, constB=0}, Alpha=@n26.R), Alpha=@n26.B) {constB=0}
  n28 = MaterialFunctionCall(Color=Abs(None=@n8)) {materialFunction=&SplitComponents}
  n29 = Constant() {r=1}
  n30 = MaterialFunctionCall(In=@n28.G, Contrast=@n29) {materialFunction=&CheapContrast}
  n31 = ComponentMask(None=@n7) {b=true, r=true}
  n32 = MaterialFunctionCall(Color=@n11) {materialFunction=&SplitComponents}
  n33 = AppendVector(A=@n32.R, B=OneMinus(None=@n32.B))
  n34 = MaterialFunctionCall(
   UVs=MaterialFunctionCall(UVs=Add(A=@n31, B=@n33), Rotation Center=Constant() {r=0.5}, "Rotation Angle (0-1)"=Constant()) {materialFunction=&CustomRotator}.Rotated Values
   Texture Scale=ComponentMask(None=@n7) {a=true}
  ) {materialFunction=&ScaleUVsByCenter}
  n35 = Multiply(A=TextureSample(UVs=@n34.UVs, Tex=@n14).RGB, B=@n34.0-1 mask)
  n36 = Constant3Vector() {constant=<rgba 0 1 1 0>}
  n37 = ComponentMask(None=@n18) {b=true, r=true}
  n38 = MaterialFunctionCall(
   UVs=MaterialFunctionCall(UVs=Add(A=@n37, B=@n33), Rotation Center=Constant() {r=0.5}, "Rotation Angle (0-1)"=Constant()) {materialFunction=&CustomRotator}.Rotated Values
   Texture Scale=ComponentMask(None=@n18) {a=true}
  ) {materialFunction=&ScaleUVsByCenter}
  n39 = Multiply(A=TextureSample(UVs=@n38.UVs, Tex=@n14).RGB, B=@n38.0-1 mask)
  n40 = Constant3Vector() {constant=<rgba 1 1 0 0>}
  n41 = ComponentMask(None=@n22) {b=true, r=true}
  n42 = MaterialFunctionCall(
   UVs=MaterialFunctionCall(UVs=Add(A=@n41, B=@n33), Rotation Center=Constant() {r=0.5}, "Rotation Angle (0-1)"=Constant()) {materialFunction=&CustomRotator}.Rotated Values
   Texture Scale=ComponentMask(None=@n22) {a=true}
  ) {materialFunction=&ScaleUVsByCenter}
  n43 = TextureSample(UVs=@n42.UVs, Tex=@n14)
  n44 = Add(
   A=Add(A=Multiply(A=@n35, B=@n36), B=Multiply(A=@n39, B=@n40))
   B=Multiply(A=Multiply(A=@n43.RGB, B=@n42.0-1 mask), B=Constant3Vector() {constant=<rgba 1 0 1 0>})
  )
  Add(A=@n25, B=Multiply(A=@n27, B=Multiply(A=@n30.Result, B=@n44)))
  n45 = Add(A=Add(A=@n12.Safe_Zone_Mask, B=@n19.Safe_Zone_Mask), B=@n23.Safe_Zone_Mask)
  n46 = MaterialFunctionCall(Color=@n4.RGB) {materialFunction=&SplitComponents}
  n47 = LinearInterpolate(A=LinearInterpolate(A=@n28.R, B=LinearInterpolate(A=@n28.B, B=@n28.B, Alpha=@n46.G), Alpha=@n46.R), B=@n28.G, Alpha=@n46.B)
  Add(
   A=Multiply(
    A=@n45
    B=MaterialFunctionCall(In=@n47, Contrast=@n29) {materialFunction=&CheapContrast}.Result
   )
   B=@n27
  )
  n48 = FunctionInput(Preview=Constant2Vector()) {bUsePreviewValueAsDefault=true, inputName=IsStates_B, inputType=FunctionInput_Vector2, sortPriority=4}
  ComponentMask(None=@n48) {g=true}
  FunctionInput() {bUsePreviewValueAsDefault=true, inputName=Object Bounds Max Adjustment, sortPriority=15}
  FunctionInput() {bUsePreviewValueAsDefault=true, inputName=Object Bounds Center Adjustment, sortPriority=15}
  FunctionInput() {bUsePreviewValueAsDefault=true, inputName=Object Bounds Min Adjustment, sortPriority=15}
  n49 = SetMaterialAttributes!(MaterialAttributes, BaseColor=Constant3Vector() {constant=<rgba 0.1626 0 1 0>}, Roughness, Specular, EmissiveColor, Metallic) {attributeSetTypes [
    #1
    #2
    #3
    #4
    #5
   ]}
  n50 = FunctionInput(Preview=@n49) {bUsePreviewValueAsDefault=true, inputName=Material, inputType=FunctionInput_MaterialAttributes, sortPriority=-20}
  n51 = GetMaterialAttributes(None=@n50) {attributeGetTypes [#1]}
  n52 = MaterialFunctionCall(Color=@n4.RGB) {materialFunction=&SplitComponents}
  n53 = LinearInterpolate(
   A=Constant3Vector() {constant=<rgba 0 0 -0.25 1>}
   B=LinearInterpolate(
    A=Constant3Vector() {constant=<rgba 0 0 0 1>}
    B=Constant3Vector() {desc=must activate TS_Vehicle_Side_01 UVs, constant=<rgba 0 0 0 1>}
    Alpha=@n52.G
   )
   Alpha=@n52.R
  )
  n54 = Constant3Vector() {constant=<rgba 0.25 0.25 0 1>}
  n55 = Constant3Vector() {constant=<rgba 0 0 0 1>}
  n56 = Constant3Vector() {desc=must activate TS_Vehicle_Side_01 UVs, constant=<rgba 0 0 -1 1>}
  n57 = MaterialFunctionCall(Color=@n4.RGB) {materialFunction=&SplitComponents}
  n58 = LinearInterpolate(
   A=LinearInterpolate(A=Constant3Vector() {constant=<rgba 0 0 -1 0>}, B=LinearInterpolate(A=@n55, B=@n56, Alpha=@n57.G), Alpha=@n57.R)
   B=Constant3Vector() {constant=<rgba 0 0 -1 0>}
   Alpha=@n57.B
  )
  n59 = Constant()
  n60 = MaterialFunctionCall(Color=@n4.RGB) {materialFunction=&SplitComponents}
  n61 = Constant3Vector() {constant=<rgba 1 0.2727 0 0>}
  n62 = FunctionInput() {bUsePreviewValueAsDefault=true, inputName="Customization Mask (_CM B)Right", sortPriority=24}
  n63 = LinearInterpolate(
   A=MaterialFunctionCall!(
    IsCharacter
    IsStatsB
    Texture=@n13
    Rotation_Axis=LinearInterpolate(A=@n53, B=@n54, Alpha=@n52.B)
    Caculated_Bounds=Add(A=@n58, B=@n11)
    Pre_Skinned_Normals=@n8
    Blend_Contrast=Constant() {r=1}
    Fill_alternate_Axis_Opacity=@n59
    "Fill_alternate_Axis_Color_A=Roughness"=Constant4Vector() {constant=<rgba 0 0 0 0.5>}.RGBA
    Is_Vehicle=@n60.R
    Is_Mech=@n60.R
    Enable_Fill_Z_Weapon
    Fill_Z_Weapon_opacity
    Fill_Z_Weapon_Color
   ) {materialFunction=&/MF_TopsSideProjection_Base}.RGB
   B=@n61
   Alpha=@n62
  )
  n64 = Saturate(None=@n45)
  n65 = LinearInterpolate(
   A=@n63
   B=LinearInterpolate(A=LinearInterpolate(A=@n1, B=@n61, Alpha=@n62), B=@n25, Alpha=@n64)
   Alpha=ScalarParameter() {group=0_UvCheker, parameterName="0_UvChecker_SafeZone   0=Off  1=On"}
  )
  n66 = MaterialFunctionCall!(
   Camera _pos
   Axis
   Scale
   Scale_Multplyer
   Object-SafeZone-Virtual-Pivot=NamedRerouteUsage()
   Object-SafeZone-Virtual-CH-Switch=NamedRerouteUsage()
   SinglePlane Normal Axis
   SinglePlane X Axis
   PlaneCenter_distortion
  ) {materialFunction=&/MF_SafeZone_VirtualPlane}
  n67 = TextureSample(UVs=@n66.Uvs, Tex=NamedRerouteUsage())
  n68 = LinearInterpolate(A=@n67.RGB, B=@n61, Alpha=Multiply(A=@n62) {constB=0.85})
  n69 = ScalarParameter() {group=0_UvCheker, parameterName="0_UvChecker_Virtual_Plane   0=Off  1=On"}
  n70 = Add(A=FunctionInput() {bUsePreviewValueAsDefault=true, inputName="Customization Mask (_CM B)left", sortPriority=24}, B=@n62)
  n71 = Constant() {r=0.5}
  n72 = SetMaterialAttributes(
   MaterialAttributes=@n51.MaterialAttributes
   BaseColor=LinearInterpolate(A=@n51.BaseColor, B=LinearInterpolate(A=@n65, B=@n68, Alpha=@n69), Alpha=Saturate(None=@n70))
   Roughness=@n71
   Specular=@n71
   Metallic=Constant()
  ) {attributeSetTypes [
    #1
    #2
    #3
    #4
    #5
   ]}
  FunctionOutput(None=StaticSwitchParameter(True=@n72, False=@n50) {group=Customization, parameterName=0_UvChecker})
  n73 = NamedRerouteUsage()
  n74 = NamedRerouteUsage()
  n75 = StaticBool() {value=true}
  n76 = Constant() {r=0.5}
  n77 = VectorParameter() {group=0_UvCheker, parameterName=0_UvChecker_Skew_offset, sortPriority=50}
  n78 = VectorParameter() {group=0_UvCheker, parameterName=0_UvChecker_Skew_Tilerate, sortPriority=50, defaultValue=<rgba 1 1 0 1>}
  n79 = TextureSample(
   UVs=LinearInterpolate(
    A=MaterialFunctionCall!(
     Normal_Map
     Offset Rt
     Pivot
     Caculated_Bounds=@n73
     Pre_Skinned_Normals=@n74
     is_State
     UV_Rotation
     Uvs_Offset
     Uvs_TileRate
     Uvs_Offset_Pivot
     For_Safe_Zones
     Rotation_Comp
     Clamp Y
     Clamp X
     Clamp Padding_Comp
     No_Rotation=@n75
    ) {materialFunction=&/MF_Projection_Side_Normal_V2}.UVs
    B=MaterialFunctionCall!(
     Normal_Map
     Caculated_Bounds=@n73
     Pre_Skinned_Normals=@n74
     Skew Amount=@n76
     Is Vehicle=@n60.R
     Uvs_Offset=@n77.RGB
     Uvs_TileRate=@n78.RGB
     Alternate Axies contrast=@n59
    ) {materialFunction=&/MF_Projection_Skew_Normal}.UVs
    Alpha=ScalarParameter() {defaultValue=1, group=0_UvCheker, parameterName="0_UvChecker_SideProjection-->Skew   1=on"}
   ) {constAlpha=1}
   Tex=NamedRerouteUsage()
  )
  LinearInterpolate(A=@n63, B=@n79.RGB, Alpha=ScalarParameter() {group=0_UvCheker, parameterName="0_UvChecker_SideProjection   0=Off  1=On"})
  n80 = Normalize(VectorInput=Subtract(A=ActorPositionWS(), B=CameraPositionWS()))
  n81 = FunctionInput(Preview=Constant3Vector() {constant=<rgba 0 0 1 1>}) {bUsePreviewValueAsDefault=true, desc="0,0,1", description="0,0,1", inputName=Axis}
  n82 = Constant3Vector()
  n83 = VectorParameter() {defaultValue=<rgba 0 40 13 1>}
  n84 = Constant() {r=0.5}
  n85 = MaterialFunctionCall() {materialFunction=/Engine/Functions/Engine_MaterialFunctions02/WorldPositionOffset/ObjectScale}
  n86 = MaterialFunctionCall(Float3=PreSkinnedLocalBounds().Extents) {materialFunction=&BreakOutFloat3Components}
  n87 = Max(A=Max(A=@n86.R, B=@n86.R), B=@n86.B)
  MaterialFunctionCall!(
   Plane Normal Axis=@n80
   Plane X Axis=Normalize(VectorInput=CrossProduct(A=@n80, B=@n81))
   Plane Center=Add(A=TransformPosition(None=@n82) {transformType=TRANSFORMPOSSOURCE_World}, B=ComponentMask(None=@n83.RGB) {b=true, g=true, r=true})
   UV Scale=Multiply(A=ComponentMask(None=@n84) {a=true}, B=Multiply(A=@n85.Scale X, B=@n87))
  ) {materialFunction=/Game/Athena/Environments/Nexus/Materials/VirtualPlaneCoordinatesAbs}
  Multiply(A=@n66.Single_plane_Masking, B=@n67.RGB)
  Multiply(A=@n15.RGB)
  n88 = Add(A=NamedRerouteUsage(), B=NamedRerouteUsage())
  Multiply(A=LinearInterpolate(B=Saturate(None=@n88), Alpha=NamedRerouteUsage()) {constA=1}, B=@n64)
  Multiply(A=@n79.RGB)
  n89 = Add(A=Add(A=@n34.0-1 mask, B=@n38.0-1 mask), B=@n42.0-1 mask)
  Multiply(A=@n27, B=Multiply(A=@n30.Result, B=Saturate(None=@n89)))
  n90 = FunctionInput(Preview=Constant4Vector() {constant=<rgba 0 0 5 -1>}.RGBA) {bUsePreviewValueAsDefault=true, inputName=Object-SafeZone-Virtual-Pivot, inputType=FunctionInput_Vector4, sortPriority=105}
  NamedRerouteDeclaration(None=@n90) {name=SafeZone VP, nodeColor=<rgba 1 0 0.8203 1>}
  n91 = FunctionInput() {bUsePreviewValueAsDefault=true, inputName=Object-SafeZone-VP-Character-PreView-Switch, inputType=FunctionInput_Scalar, sortPriority=105}
  NamedRerouteDeclaration(None=@n91) {name=SafeZone Character-Switch, nodeColor=<rgba 1 0 0.8203 1>}
  n92 = ComponentMask(None=Add(A=@n11) {constB=-0.5}) {g=true}
  NamedRerouteDeclaration(None=Saturate(None=Divide(A=@n92) {constB=-0.001})) {name=Character mask, nodeColor=<rgba 0.9453 1 0 1>}
  NamedRerouteDeclaration(None=@n3.A) {name=Is_Character, nodeColor=<rgba 0 0.25 1 1>}
  NamedRerouteDeclaration(None=ComponentMask(None=@n48) {r=true}) {name=Is_Back, nodeColor=<rgba 0 0.25 1 1>}
  NamedRerouteDeclaration(None=@n11) {name=Cal Bounds, nodeColor=<rgba 0.2656 0 1 1>}
  NamedRerouteDeclaration(None=@n8) {name=Preskin_Uvs, nodeColor=<rgba 0.0547 1 0 1>}
  NamedRerouteUsage()
 ]}
