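# Material function for the Galaxy weapon wrap, blending an animated emissive galaxy look into incoming material attributes by mask, behind the Wrap_Galaxy static switch.
# Camera-vector triplanar projections layer a rotating nebula with small and twinkling bright stars, then a Fresnel rim broken up by pre-skinned swirl noise.
# An additive glow uses a bounds gradient, or a rotated texture for vehicles, with mech-specific mapping.
# The galaxy layer has black base color and keeps the input normal.
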MaterialFunction /Game/Weapons/WeaponSkins/Wraps/MaterialFunctions/MF_Wrap_Galaxy
graph {nodes [
  n1 = TextureCoordinate() {uTiling=2, vTiling=2}
  n2 = ScalarParameter() {defaultValue=1, group=Wrap_Galaxy, parameterName=WG_Edge Fresnel Tiling}
  n3 = ScalarParameter() {group=Wrap_Galaxy, parameterName=WG_GradientSpeed_X}
  n4 = ScalarParameter() {defaultValue=0.1, group=Wrap_Galaxy, parameterName=WG_GradientSpeed_Y}
  n5 = TextureSample(UVs=Panner(Coordinate=Multiply(A=@n1, B=@n2), Speed=AppendVector(A=@n3, B=@n4))) {
   samplerSource=SSM_Wrap_WorldGroupSettings
   samplerType=SAMPLERTYPE_LinearColor
   texture=/Game/Weapons/WeaponSkins/Textures/T_Swirly_Noise_highContrast1
  }
  FunctionInput(Preview=@n5.RGB) {bUsePreviewValueAsDefault=true, inputName=GalaxyAlternateFresnel, inputType=FunctionInput_Scalar}
  n6 = FunctionInput() {inputName=Material, inputType=FunctionInput_MaterialAttributes, sortPriority=-10}
  n7 = NamedRerouteUsage()
  n8 = NamedRerouteUsage()
  n9 = LinearInterpolate(
   B=ComponentMask(None=FunctionInput() {inputName=Custom MAsk, inputType=FunctionInput_Vector4}) {r=true}
   Alpha=Saturate(None=Subtract(A=@n7, B=@n8))
  ) {constA=1}
  n10 = GetMaterialAttributes(None=@n6) {attributeGetTypes [#1, #2]}
  n11 = VectorParameter() {group=Wrap_Galaxy, parameterName="WG_Galaxy_Glow_Color_A=Brightness", defaultValue=<rgba 0 0 0 1>}
  n12 = ScalarParameter() {group=Wrap_Galaxy, parameterName=WG_Galaxy_Glow_Offset_Weapon}
  n13 = ScalarParameter() {group=Wrap_Galaxy, parameterName=WG_Galaxy_Glow_Falloff_Weapon}
  n14 = VertexInterpolator(VS=PreSkinnedPosition())
  n15 = PreSkinnedLocalBounds()
  n16 = Divide(A=Subtract(A=@n14.PS, B=@n15.Min), B=Subtract(A=@n15.Max, B=@n15.Max))
  n17 = MaterialFunctionCall!(
   Bounds=@n16
   ForwardAxis=FunctionInput(Preview=Constant3Vector() {constant=<rgba 1 0 0 0>}) {bUsePreviewValueAsDefault=true, inputName=ForwardAxis}
   Offset=@n12
   FallOff=@n13
  ) {materialFunction=&/MF_Wrap_Gradent_Simple}
  n18 = FunctionInput(Preview=Constant()) {bUsePreviewValueAsDefault=true, inputName=isVehicle, inputType=FunctionInput_Scalar, sortPriority=5}
  n19 = VectorParameter() {group=Wrap_Galaxy, parameterName=WG_Galaxy_Glow_TEXOffset_vehicle, sortPriority=200, defaultValue=<rgba 0 0 0 1>}
  n20 = VectorParameter() {group=Wrap_Galaxy, parameterName=WG_Galaxy_Glow_TEXOffset_vehicle_SP, sortPriority=200, defaultValue=<rgba 0 0 0 1>}
  n21 = Constant()
  n22 = FunctionInput(Preview=@n21) {bUsePreviewValueAsDefault=true, inputName=Side Project Vehicle, inputType=FunctionInput_Scalar}
  n23 = LinearInterpolate(A=Constant(), B=LinearInterpolate(A=@n19.RGB, B=@n20.RGB, Alpha=@n22), Alpha=@n18)
  n24 = VectorParameter() {group=Wrap_Galaxy, parameterName=WG_Galaxy_Glow_TEXOffset_Mech, sortPriority=200, defaultValue=<rgba 0 0 0 1>}
  n25 = FunctionInput(Preview=@n21) {bUsePreviewValueAsDefault=true, inputName=isMech, inputType=FunctionInput_Scalar, sortPriority=4}
  n26 = FunctionInput(Preview=@n16) {bUsePreviewValueAsDefault=true, inputName=Calcated_Bounds, sortPriority=11}
  n27 = Constant3Vector() {constant=<rgba 1 1 -1 0>}
  n28 = MaterialFunctionCall(Color=Add(A=LinearInterpolate(A=@n23, B=@n24.RGB, Alpha=@n25), B=Multiply(A=@n26, B=@n27))) {materialFunction=&SplitComponents}
  n29 = Multiply(
   A=AppendVector(A=@n28.R, B=@n28.R)
   B=ScalarParameter() {defaultValue=1, group=Wrap_Galaxy, parameterName=WG_Galaxy_Glow_TEXScale_vehicle, sortPriority=200}
  )
  n30 = AppendVector(A=@n28.R, B=@n28.R)
  n31 = LinearInterpolate(
   A=LinearInterpolate(A=AppendVector(A=@n28.G, B=@n28.G), B=LinearInterpolate(A=@n29, B=@n30, Alpha=@n22), Alpha=@n18)
   B=AppendVector(A=@n28.R, B=@n28.R)
   Alpha=@n25
  )
  n32 = Constant() {r=0.5}
  n33 = MaterialFunctionCall(Float4=NamedRerouteUsage()) {materialFunction=&BreakOutFloat4Components}
  n34 = LinearInterpolate(
   A=LinearInterpolate(B=ScalarParameter() {group=Wrap_Galaxy, parameterName=WG_Galaxy_Glow_TEXRrotation_vehicle, sortPriority=200}, Alpha=@n33.R)
   B=ScalarParameter() {group=Wrap_Galaxy, parameterName=WG_Galaxy_Glow_TEXRrotation_Mech, sortPriority=200}
   Alpha=@n33.B
  )
  n35 = TextureObjectParameter() {group=Wrap_Galaxy, parameterName=WG_Galaxy_Glow_Texture_Vehicle, sortPriority=199, texture=&T_BlacksRGB}
  n36 = ChannelMaskParameter(None=TextureSample(
    UVs=MaterialFunctionCall(UVs=@n31, Rotation Center=@n32, "Rotation Angle (0-1)"=@n34) {materialFunction=&CustomRotator}.Rotated Values
    Tex=@n35
   ).RGB) {group=Wrap_Galaxy, parameterName=WG_Galaxy_Glow_Texture_Channel, sortPriority=199}
  n37 = Multiply(A=@n11.RGB, B=Saturate(None=LinearInterpolate(A=@n17.Raw_Data, B=@n36.R, Alpha=@n18)))
  n38 = ScalarParameter() {defaultValue=1, group=Wrap_Galaxy, parameterName=WG_Galaxy_GlobalBrightness}
  n39 = Time()
  n40 = Multiply(A=@n39, B=Constant())
  n41 = ScalarParameter() {group=Wrap_Galaxy, parameterName=WG_Background Rotation}
  n42 = AppendVector(
   A=ScalarParameter() {defaultValue=0.02, group=Wrap_Galaxy, parameterName=WG_Galaxy Rotation Speed}
   B=ScalarParameter() {defaultValue=-0.02, group=Wrap_Galaxy, parameterName=WG_Galaxy Rotation Speed_1}
  )
  n43 = ScalarParameter() {defaultValue=0.1, group=Wrap_Galaxy, parameterName=WG_Galaxy Rotation Speed_2}
  n44 = TextureObjectParameter() {
   group=Wrap_Galaxy
   parameterName=WG_GalaxyTexture
   texture=/Game/Characters/Player/Male/Medium/Bodies/M_Med_Soldier_04/Skins/Galaxy/Textures/T_FN_Nebula
  }
  n45 = MaterialFunctionCall!(
   CameraVector
   Rotation Speed=Multiply(A=LinearInterpolate(A=@n39, B=@n40, Alpha=@n41), B=AppendVector(A=@n42, B=@n43))
   Texture=@n44
   Tiling=ScalarParameter() {defaultValue=0.5, group=Wrap_Galaxy, parameterName=WG_GalaxyTiling}
   Offset=Multiply(
    A=FunctionInput() {bUsePreviewValueAsDefault=true, inputName=Warp Mask, inputType=FunctionInput_Scalar}
    B=ScalarParameter() {defaultValue=0.012, group=Wrap_Galaxy, parameterName=WG_WarpIntensity}
   )
   AxisFadeContrast=ScalarParameter() {defaultValue=6, group=Wrap_Galaxy, parameterName=WG_GalaxyAxisFade}
  ) {materialFunction=/Game/Weapons/WeaponSkins/MaterialFunction/TriplanarCameraVector_Rotated}
  n46 = Desaturation(None=@n45.Result, Fraction=ScalarParameter() {defaultValue=0.5, group=Wrap_Galaxy, parameterName=WG_Star Saturation})
  n47 = Multiply(A=@n46, B=VectorParameter() {group=Wrap_Galaxy, parameterName=WG_Star Brightness, defaultValue=<rgba 300 300 300 1>}.RGB)
  n48 = Time()
  n49 = ScalarParameter() {defaultValue=-0.125, group=Wrap_Galaxy, parameterName=WG_Small Star Rotation speed}
  n50 = TextureObjectParameter() {
   parameterName=Stars
   samplerType=SAMPLERTYPE_Masks
   texture=/Game/Characters/Player/Male/Medium/Bodies/M_Med_Soldier_04/Skins/Galaxy/Textures/Celestial_T_Stars
  }
  n51 = Time()
  n52 = VectorParameter() {group=Wrap_Galaxy, parameterName=WG_Small Star_Pan_Speed, defaultValue=<rgba 0 0 0 1>}
  n53 = ComponentMask(None=MaterialFunctionCall(
    Rotation Speed=Multiply(A=@n48, B=@n49)
    Texture=@n50
    Tiling=ScalarParameter() {defaultValue=5, group=Wrap_Galaxy, parameterName=WG_Small Star Tiling}
    Offset=Multiply(A=@n51, B=@n52.RGB)
   ) {materialFunction=&TriplanarCameraVector_Rotated}.Result) {b=true}
  n54 = Multiply(A=@n53, B=@n53)
  n55 = VectorParameter() {group=Wrap_Galaxy, parameterName=WG_Star Brightness_1, defaultValue=<rgba 950 950 950 1>}
  n56 = ComponentMask(None=MaterialFunctionCall!(
    CameraVector
    Texture=@n50
    Tiling=ScalarParameter() {defaultValue=2, group=Wrap_Galaxy, parameterName=WG_Bright Star Tiling}
    Offset
    AxisFadeContrast
   ) {materialFunction=/Engine/Functions/Engine_MaterialFunctions01/Texturing/TriplanarCameraVector}.Result) {r=true}
  n57 = Multiply(A=@n56, B=@n56)
  n58 = TextureSample(UVs=Panner(Coordinate=TextureCoordinate() {uTiling=2, vTiling=2}) {speedX=0.133, speedY=0.2}) {
   samplerSource=SSM_Wrap_WorldGroupSettings
   samplerType=SAMPLERTYPE_Masks
   texture=&Celestial_T_Stars
  }
  n59 = ScalarParameter() {defaultValue=8, group=Wrap_Galaxy, parameterName=WG_AccentRimBrightness}
  n60 = VectorParameter() {group=Wrap_Galaxy, parameterName=WG_GalaxyEF, channelNames {a="", b=GspeedY, g=GspeedX, r=EdgeFresTile}, defaultValue=<rgba 200 0 0.1 1>}
  n61 = AppendVector(A=@n60.GspeedX, B=@n60.GspeedX)
  n62 = PreSkinnedNormal()
  n63 = ComponentMask(None=MaterialFunctionCall!(
    TextureObject=TextureObject() {samplerType=SAMPLERTYPE_LinearColor, texture=&T_Swirly_Noise_highContrast1}
    TextureSize=@n60.EdgeFresTile
    WorldPosition=@n14.PS
    Export Float 4
    In=@n61
    World Space Normal=VertexInterpolator(VS=@n62).PS
    ProjectionTransitionContrast
   ) {materialFunction=/Game/Athena/SafeZone/Preskinned_WP_Texture_ScalePan}.XYZ Texture) {r=true}
  n64 = ScalarParameter() {defaultValue=1, group=Wrap_Galaxy, parameterName=WG_Galaxy Flame Fresnel Intensity}
  n65 = ScalarParameter() {defaultValue=5, group=Wrap_Galaxy, parameterName=WG_Galaxy_FresEX}
  n66 = LinearInterpolate(
   A=LinearInterpolate(A=LinearInterpolate(A=@n45.Result, B=@n47, Alpha=@n54), B=Multiply(A=@n46, B=@n55.RGB), Alpha=Multiply(A=@n57, B=@n58.G))
   B=LinearInterpolate(
    A=Multiply(A=@n59, B=@n45.Result)
    B=VectorParameter() {group=Wrap_Galaxy, parameterName=EdgeFresnel, defaultValue=<rgba 5.0909 8.586 10 1>}.RGB
    Alpha=Multiply(A=@n63, B=@n64)
   )
   Alpha=Saturate(None=Fresnel(ExponentIn=@n65) {baseReflectFraction=0})
  )
  n67 = SetMaterialAttributes!(
   MaterialAttributes
   BaseColor=Constant()
   Specular=ScalarParameter() {group=Wrap_Galaxy, parameterName=WG_Galaxy Specular}
   Roughness=Multiply(A=@n10.Roughness, B=ScalarParameter() {defaultValue=1, group=Wrap_Galaxy, parameterName=WG_Galaxy Roughness})
   EmissiveColor=Add(A=Multiply(A=@n11.A, B=@n37), B=Multiply(A=@n38, B=@n66))
   Normal=@n10.Normal
  ) {attributeSetTypes [
    #3
    #4
    #2
    #5
    #1
   ]}
  FunctionOutput(None=StaticSwitchParameter(True=BlendMaterialAttributes(A=@n6, B=@n67, Alpha=@n9) {vertexAttributeBlendType=UseA}, False=@n6) {group=Customization_Wrap, parameterName=Wrap_Galaxy}) {outputName=Result_1}
  LinearInterpolate(A=@n16, B=@n26, Alpha=@n22)
  n68 = ScalarParameter() {group=Wrap_Galaxy, parameterName=WG_Galaxy_Glow_Offset_Vehicle}
  n69 = ScalarParameter() {group=Wrap_Galaxy, parameterName=WG_Galaxy_Glow_Offset_side_Vehicle}
  LinearInterpolate(
   A=LinearInterpolate(A=@n12, B=LinearInterpolate(A=@n68, B=@n69, Alpha=@n22), Alpha=@n18)
   B=ScalarParameter() {group=Wrap_Galaxy, parameterName=WG_Galaxy_Glow_Offset_Mech}
   Alpha=@n25
  )
  LinearInterpolate(
   A=LinearInterpolate(A=@n13, B=ScalarParameter() {group=Wrap_Galaxy, parameterName=WG_Galaxy_Glow_Falloff_Vehicle}, Alpha=@n18)
   B=ScalarParameter() {group=Wrap_Galaxy, parameterName=WG_Galaxy_Glow_Falloff_Mech}
   Alpha=@n25
  )
  n70 = FunctionInput() {bUsePreviewValueAsDefault=true, inputName=isCharacter, inputType=FunctionInput_Scalar, sortPriority=4}
  NamedRerouteDeclaration(None=MaterialFunctionCall(X=@n18, Y=@n22, Z=@n25, A=@n70) {materialFunction=&MakeFloat4}.Result) {name=IsStates, nodeColor=<rgba 1 0.9141 0 1>}
  NamedRerouteDeclaration(None=@n70) {name=IsCharacter, nodeColor=<rgba 0.9957 1 0 1>}
  n71 = FunctionInput() {bUsePreviewValueAsDefault=true, inputName=IsStatsB, inputType=FunctionInput_Vector4, sortPriority=-19}
  NamedRerouteDeclaration(None=ComponentMask(None=@n71) {r=true}) {name=IsBackpack, nodeColor=<rgba 0 0.9253 1 1>}
  NamedRerouteDeclaration(None=ComponentMask(None=@n71) {a=true}) {name=isCharAccent, nodeColor=<rgba 0 0.9253 1 1>}
  NamedRerouteDeclaration(None=ComponentMask(None=@n71) {g=true}) {name=IsPickaxe, nodeColor=<rgba 0 0.9253 1 1>}
  OneMinus(None=@n28.B)
  TextureObject() {samplerType=SAMPLERTYPE_Masks, texture=&Celestial_T_Stars}
  n72 = TextureCoordinate() {uTiling=2, vTiling=2}
  TextureSample(UVs=Panner(Coordinate=Multiply(A=@n72, B=@n60.EdgeFresTile), Speed=@n61)) {
   samplerSource=SSM_Wrap_WorldGroupSettings
   samplerType=SAMPLERTYPE_LinearColor
   texture=&T_Swirly_Noise_highContrast1
  }
 ]}
