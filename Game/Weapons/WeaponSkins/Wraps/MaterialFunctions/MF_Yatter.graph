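# Material function for the Yatter weapon wrap, covering vehicles and other items, blended over incoming material attributes by customization masks behind a static switch.
# Pre-skinned position and bounds form planar UVs, with pickaxe, character and backpack adjustments and switchable vehicle projection.
# These UVs drive a stripe texture, or procedural stepped bands on vehicles.
# MF_ProgressBar adds animated colored, emissive streaks: two generally, two more on vehicles.
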MaterialFunction /Game/Weapons/WeaponSkins/Wraps/MaterialFunctions/MF_Yatter
graph {nodes [
  n1 = FunctionInput(Preview=PreSkinnedLocalBounds().Extents) {bUsePreviewValueAsDefault=true, inputName=ObjectBounds}
  n2 = MaterialFunctionCall(Float3=@n1) {materialFunction=&BreakOutFloat3Components}
  AppendVector(A=@n2.G, B=@n2.G)
  AppendVector(A=@n2.R, B=@n2.R)
  AppendVector(A=@n2.R, B=@n2.R)
  n3 = FunctionInput() {inputName=Material Attributes, inputType=FunctionInput_MaterialAttributes}
  n4 = PreSkinnedPosition()
  n5 = Subtract(
   A=FunctionInput(Preview=VertexInterpolator(VS=@n4).PS) {bUsePreviewValueAsDefault=true, inputName=WorldPosition}
   B=Multiply(A=@n1) {constB=0.5}
  ) {constB=100}
  n6 = Add(A=@n5, B=VectorParameter() {parameterName=PliantBoundAdjust_Shift_Pickaxe}.RGB)
  n7 = VectorParameter() {parameterName=PliantBoundAdjust_ScaleUV_Pickaxe, defaultValue=<rgba 1 1 1 0>}
  n8 = Add(A=@n5, B=VectorParameter() {parameterName=PliantBoundAdjust_Shift_Character}.RGB)
  n9 = VectorParameter() {parameterName=PliantBoundAdjust_ScaleUV_Character, defaultValue=<rgba 1 1 1 0>}
  n10 = Add(A=@n5, B=VectorParameter() {parameterName=PliantBoundAdjust_Shift_Backpack}.RGB)
  n11 = VectorParameter() {parameterName=PliantBoundAdjust_ScaleUV_Backpack, defaultValue=<rgba 1 1 1 0>}
  n12 = MaterialFunctionCall(Float3=LinearInterpolate(A=Multiply(A=@n8, B=@n9.RGB), B=Multiply(A=@n10, B=@n11.RGB), Alpha=NamedRerouteUsage())) {materialFunction=&BreakOutFloat3Components}
  n13 = LinearInterpolate(
   A=LinearInterpolate(A=@n5, B=Multiply(A=@n6, B=@n7.RGB), Alpha=NamedRerouteUsage())
   B=MaterialFunctionCall(X=@n12.R, Y=@n12.R, Z=@n12.R) {materialFunction=&MakeFloat3}.Result
   Alpha=NamedRerouteUsage()
  )
  n14 = Abs(None=ComponentMask(None=@n1) {b=true, g=true})
  n15 = ComponentMask(None=@n13) {g=true, r=true}
  n16 = Multiply(A=Abs(None=ComponentMask(None=@n1) {g=true, r=true})) {constB=-1}
  n17 = ComponentMask(None=@n13) {b=true, r=true}
  n18 = Multiply(A=Abs(None=ComponentMask(None=@n1) {b=true, r=true})) {constB=-1}
  n19 = Round(None=Saturate(None=FunctionInput() {bUsePreviewValueAsDefault=true, inputName=SideProjectVehicle, inputType=FunctionInput_Scalar}))
  n20 = Round(None=Saturate(None=FunctionInput() {inputName=isVehicle, inputType=FunctionInput_Scalar}))
  n21 = LinearInterpolate(
   A=Divide(A=ComponentMask(None=@n13) {b=true, g=true}, B=Multiply(A=@n14) {constB=-1})
   B=LinearInterpolate(A=Divide(A=@n15, B=@n16), B=Divide(A=@n17, B=@n18), Alpha=@n19)
   Alpha=@n20
  )
  n22 = VectorParameter() {group=Yatter, parameterName=YtterVehicleStipesColor0, defaultValue=<rgba 0.224 0.224 0.224 1>}
  n23 = VectorParameter() {group=Yatter, parameterName=YatterVehicleStipesColor1, defaultValue=<rgba 0.0104 0.0104 0.0104 1>}
  MaterialExpressionComponentMask_1 = ComponentMask(None=@n21) {g=true}
  n24 = OneMinus(None=@MaterialExpressionComponentMask_1)
  n25 = ScalarParameter() {defaultValue=10, group=Yatter, parameterName=StripesCount}
  n26 = Divide(A=Floor(None=Multiply(A=@n24, B=@n25)), B=@n25)
  MaterialExpressionScalarParameter_25 = ScalarParameter() {defaultValue=0.7, group=Yatter, parameterName=StripesMaskMax}
  n27 = Custom!(a=@MaterialExpressionComponentMask_1, b=@MaterialExpressionScalarParameter_25) {
   code="return step(a,b);"
   description=Step
   outputType=CMOT_Float1
   inputs [
    {inputName=a, input {expression=&self:MaterialExpressionComponentMask_1}}
    {inputName=b, input {expression=&self:MaterialExpressionScalarParameter_25}}
   ]
  }
  MaterialExpressionScalarParameter_24 = ScalarParameter() {defaultValue=0.4, group=Yatter, parameterName=StripesMaskMin}
  n28 = Custom(a=@MaterialExpressionComponentMask_1, b=@MaterialExpressionScalarParameter_24) {
   code="return step(a,b);"
   description=Step
   outputType=CMOT_Float1
   inputs [
    {inputName=a, input {expression=&self:MaterialExpressionComponentMask_1}}
    {inputName=b, input {expression=&self:MaterialExpressionScalarParameter_24}}
   ]
  }
  n29 = LinearInterpolate(
   A=TextureSample(UVs=@n21, Tex=TextureObjectParameter() {group=Yatter, parameterName=Yatter_Msk, texture=&T_BlacksRGB}) {samplerSource=SSM_Clamp_WorldGroupSettings, texture=/Game/Weapons/WeaponSkins/Wraps/Textures/Yatter/T_YatterStripes}.RGB
   B=LinearInterpolate(
    A=VectorParameter() {group=Yatter, parameterName=YatterVehicleBaseColor, defaultValue=<rgba 1 1 1 1>}.RGB
    B=LinearInterpolate(A=@n22.RGB, B=@n23.RGB, Alpha=@n26)
    Alpha=Subtract(A=@n27, B=@n28)
   )
   Alpha=@n20
  )
  n30 = VectorParameter() {group=Yatter, parameterName=YatterPanningStripesColor, defaultValue=<rgba 1 0 0 1>}
  n31 = ScalarParameter() {defaultValue=0.17, group=Yatter, parameterName=OffsetLine1}
  n32 = LinearInterpolate(
   A=ScalarParameter() {defaultValue=0.55, group=Yatter, parameterName=OffsetLine1Vehicle}
   B=ScalarParameter() {defaultValue=0.1, group=Yatter, parameterName=OffsetLine1VehicleSide}
   Alpha=@n19
  )
  n33 = LinearInterpolate(
   A=ScalarParameter() {defaultValue=5, group=Yatter, parameterName=PanningStripesTailContrast}
   B=ScalarParameter() {defaultValue=4, group=Yatter, parameterName=PanningStripesTailContrastVehicles}
   Alpha=@n20
  )
  n34 = ScalarParameter() {defaultValue=0.25, group=Yatter, parameterName=OffsetLine2}
  n35 = LinearInterpolate(
   A=ScalarParameter() {defaultValue=0.4, group=Yatter, parameterName=OffsetLine2Vehicle}
   B=ScalarParameter() {defaultValue=-0.2, group=Yatter, parameterName=OffsetLine2VehicleSide}
   Alpha=@n19
  )
  n36 = Add(
   A=MaterialFunctionCall!(
    Thickness=ScalarParameter() {defaultValue=0.05, group=Yatter, parameterName=ThicknessLine1}
    TexCoord=@n21
    ProgressBarTipAngle=ScalarParameter() {defaultValue=11, group=Yatter, parameterName=PanningStripesTipAngleL1}
    ProgressBarTailContrast=@n33
    Speed=ScalarParameter() {defaultValue=1, group=Yatter, parameterName=SpeedLine1}
    VerticalOffset=LinearInterpolate(A=@n31, B=@n32, Alpha=@n20)
   ) {materialFunction=&/MF_ProgressBar}.Output
   B=MaterialFunctionCall(
    Thickness=ScalarParameter() {defaultValue=0.08, group=Yatter, parameterName=ThicknessLine2}
    TexCoord=@n21
    ProgressBarTipAngle=ScalarParameter() {defaultValue=15, group=Yatter, parameterName=PanningStripesTipAngle}
    ProgressBarTailContrast=@n33
    Speed=ScalarParameter() {defaultValue=0.5, group=Yatter, parameterName=SpeedLine2}
    VerticalOffset=LinearInterpolate(A=@n34, B=@n35, Alpha=@n20)
   ) {desc=Top, materialFunction=&MF_ProgressBar}.Output
  )
  n37 = ScalarParameter() {defaultValue=3, group=Yatter, parameterName=PanningStripesTailContrastVehiclesL3}
  n38 = ScalarParameter() {defaultValue=0.7, group=Yatter, parameterName=OffsetLine3Vehicle}
  n39 = ScalarParameter() {defaultValue=0.2, group=Yatter, parameterName=OffsetLine3VehicleSide}
  n40 = Multiply(
   A=MaterialFunctionCall(
    Thickness=ScalarParameter() {defaultValue=0.4, group=Yatter, parameterName=ThicknessLine3}
    TexCoord=@n21
    ProgressBarTipAngle=ScalarParameter() {defaultValue=15, group=Yatter, parameterName=PanningStripesTipAngleL3}
    ProgressBarTailContrast=@n37
    Speed=ScalarParameter() {defaultValue=0.7, group=Yatter, parameterName=SpeedLine3}
    VerticalOffset=LinearInterpolate(A=@n38, B=@n39, Alpha=@n19)
   ) {desc=Top, materialFunction=&MF_ProgressBar}.Output
   B=@n20
  )
  n41 = MaterialFunctionCall(
   Thickness=ScalarParameter() {defaultValue=0.4, group=Yatter, parameterName=ThicknessLine4}
   TexCoord=@n21
   ProgressBarTipAngle=ScalarParameter() {defaultValue=60, group=Yatter, parameterName=PanningStripesTipAngleL4}
   ProgressBarTailContrast=@n37
   Speed=ScalarParameter() {defaultValue=0.5, group=Yatter, parameterName=SpeedLine4}
   VerticalOffset=LinearInterpolate(
    A=ScalarParameter() {defaultValue=-0.08, group=Yatter, parameterName=OffsetLine4Vehicle}
    B=ScalarParameter() {defaultValue=-0.4, group=Yatter, parameterName=OffsetLine4VehicleSide}
    Alpha=@n19
   )
  ) {desc=Top, materialFunction=&MF_ProgressBar}
  n42 = Saturate(None=Add(A=Add(A=@n36, B=@n40) {constB=0}, B=Multiply(A=@n41.Output, B=@n20)) {constB=0})
  n43 = FunctionInput() {bUsePreviewValueAsDefault=true, inputName="Customization Mask 2 (_CM B Top Right)", inputType=FunctionInput_Scalar, sortPriority=25}
  n44 = Multiply(A=Multiply(A=@n30.RGB, B=@n42), B=ScalarParameter() {defaultValue=0.2, group=Yatter, parameterName=StripesEmissive})
  n45 = FunctionInput() {bUsePreviewValueAsDefault=true, inputName="Customization Mask 1 (_CM B Top Left)", inputType=FunctionInput_Scalar, sortPriority=25}
  n46 = SetMaterialAttributes!(
   MaterialAttributes=@n3
   BaseColor=LinearInterpolate(
    A=LinearInterpolate(A=@n29, B=@n30.RGB, Alpha=@n42)
    B=VectorParameter() {group=Yatter, parameterName=YatterMaskedPartsColor, defaultValue=<rgba 0.0467 0.0467 0.0467 1>}.RGB
    Alpha=@n43
   ) {desc=BaseColor}
   EmissiveColor=Multiply(A=Multiply(A=@n44, B=@n45), B=OneMinus(None=@n43))
   Metallic=Constant()
  ) {attributeSetTypes [#1, #2, #3]}
  n47 = Saturate(None=Add(A=@n45, B=@n43))
  FunctionOutput(None=StaticSwitchParameter(True=BlendMaterialAttributes(A=@n3, B=@n46, Alpha=@n47), False=@n3) {group=Customization_Wrap, parameterName=Yatter Wrap})
  n48 = FunctionInput(Preview=VertexInterpolator(VS=PreSkinnedNormal()).PS) {bUsePreviewValueAsDefault=true, inputName=Normals, sortPriority=1, previewValue=<0 0 1 1>}
  n49 = ComponentMask(None=@n48) {r=true}
  n50 = Abs(None=ComponentMask(None=@n48) {b=true})
  n51 = Abs(None=ComponentMask(None=@n48) {g=true})
  Multiply(A=@n42, B=LinearInterpolate(A=Abs(None=@n49), B=LinearInterpolate(A=@n50, B=@n51, Alpha=@n19), Alpha=@n20))
  NamedRerouteDeclaration(None=FunctionInput() {bUsePreviewValueAsDefault=true, inputName=IsCharacter, inputType=FunctionInput_Scalar}) {name=IsCharacter, nodeColor=<rgba 0 0.9253 1 1>}
  n52 = FunctionInput() {bUsePreviewValueAsDefault=true, inputName=IsStatsB, inputType=FunctionInput_Vector2}
  NamedRerouteDeclaration(None=ComponentMask(None=@n52) {r=true}) {name=IsBackpack, nodeColor=<rgba 0 0.9253 1 1>}
  NamedRerouteDeclaration(None=ComponentMask(None=@n52) {g=true}) {name=IsPickaxe, nodeColor=<rgba 0 0.9253 1 1>}
 ]}
