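# Master material function for wrap customization, layering cosmetic effects onto weapon, vehicle, mech, character, backpack and pickaxe materials.
# Shares masks, object states, bounds and safe zones through named reroutes.
# Primary or Primary_V2 base functions feed long chains of effect functions, with static switches selecting the V2, older or One_Offs paths.
# A final effects pass feeds a vertex-alpha wrap holdout, and Allow Customization Effects can bypass everything.
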MaterialFunction /Game/Packages/Fortress_SharedMaterials/Base_Material_Functions/BaseCustomizationFunctions/MF_Base_Customization
graph {nodes [
  FunctionInput() {bUsePreviewValueAsDefault=true, inputName=--------------------, inputType=FunctionInput_Scalar, sortPriority=40}
  FunctionInput() {bUsePreviewValueAsDefault=true, inputName="Customization Mask 3 (_CM B Bottom Left)", inputType=FunctionInput_Scalar, sortPriority=26}
  FunctionInput() {bUsePreviewValueAsDefault=true, inputName=-------------------, inputType=FunctionInput_Scalar, sortPriority=23}
  FunctionInput() {bUsePreviewValueAsDefault=true, inputName=------------------, inputType=FunctionInput_Scalar, sortPriority=-19}
  n1 = NamedRerouteUsage()
  n2 = NamedRerouteUsage()
  n3 = MaterialFunctionCall!(
   Material=NamedRerouteUsage()
   ------------------
   Object SBounds Min Adjustment=NamedRerouteUsage()
   Object Bounds Center Adjustment=NamedRerouteUsage()
   Object Bounds Max Adjustment=NamedRerouteUsage()
   Wrap ScaleMultiplier=NamedRerouteUsage()
   Rotation Z Axis=NamedRerouteUsage()
   Forward Axis=NamedRerouteUsage()
   Rotation X Axis=NamedRerouteUsage()
   IsStates=NamedRerouteUsage()
   IsStatesB V4=NamedRerouteUsage()
   IsStatesB=NamedRerouteUsage()
   "Customization Mask (_CM RGBA)"=NamedRerouteUsage()
   --------------------
   Pre-Skinned Local Position=NamedRerouteUsage()
   Pre-Skinned Local Normal=NamedRerouteUsage()
   Pre-Skinned Bounds=NamedRerouteUsage()
   Object-SafeZone-Back=NamedRerouteUsage()
   Object-SafeZone-Middle=NamedRerouteUsage()
   Object-SafeZone-Front=NamedRerouteUsage()
   Object-SafeZone-Virtual-Pivot=NamedRerouteUsage()
   Object-SafeZone-VP_Charcter_Preview_Switch=NamedRerouteUsage()
  ) {materialFunction=&/MF_Base_Customization_Primary}
  n4 = MaterialFunctionCall!(
   Material=MaterialFunctionCall!(
    Material=@n3.Result
    ------------------
    IsVehicle=NamedRerouteUsage()
    is_StateB=NamedRerouteUsage()
    Normal Map=NamedRerouteUsage()
    "Customization Crevasse (_CM G)"=NamedRerouteUsage()
    "Customization Mask UV Islands (_CM B Bottom Right)"=NamedRerouteUsage()
    -------------------
    "Customization Mask (_CM B)"=NamedRerouteUsage()
    --------------------
    Pre-Skinned Local Position=NamedRerouteUsage()
    Pre-Skinned Local Normal=NamedRerouteUsage()
    isCharAccent=NamedRerouteUsage()
   ) {materialFunction=/Game/Weapons/WeaponSkins/Wraps/MaterialFunctions/MF_Customization_Simple}.Result
   Topside 02 Msk=NamedRerouteUsage()
   Topside 01 Msk=NamedRerouteUsage()
   Triplaner 01 Mask=@n3.Triplaner Mask
   Triplaner 02 Alpha=@n3.Triplaner 02 Alpha
   Triplaner 01 Alpha=@n3.Triplaner 01 Alpha
   Side Project Vehicle=NamedRerouteUsage()
   IsStatsB=NamedRerouteUsage()
   ForwardAxis=NamedRerouteUsage()
   Local Normals=NamedRerouteUsage()
   IsCharacter=NamedRerouteUsage()
   Local Position=NamedRerouteUsage()
   IsState=NamedRerouteUsage()
   "Customization Crevasse (_CM G)"=NamedRerouteUsage()
   "Customization Mask (_CM B Top left)"=NamedRerouteUsage()
   "Customization Mask 2 (_CM B Top Right)"=NamedRerouteUsage()
   isVehicle=NamedRerouteUsage()
   isMech=NamedRerouteUsage()
   Object Bounds Center Adjustment=NamedRerouteUsage()
   Object Bounds Min Adjustment=NamedRerouteUsage()
   Object Bounds Max Adjustment=NamedRerouteUsage()
   Object-SafeZone-VP_Charcter_Preview_Switch=NamedRerouteUsage()
   Object-SafeZone-Virtual-Pivot=NamedRerouteUsage()
  ) {materialFunction=/Game/Weapons/WeaponSkins/Wraps/MaterialFunctions/MF_Curvture_Emissive}
  n5 = StaticSwitch(True=NamedRerouteUsage(), False=@n4.Result, Value=NamedRerouteUsage())
  n6 = MaterialFunctionCall!(
   Material=@n5
   Mask Crevise=NamedRerouteUsage()
   Mask Primary=NamedRerouteUsage()
   Mask Secondary=NamedRerouteUsage()
   TS_Mask_02=NamedRerouteUsage()
   TS_Mask_01=NamedRerouteUsage()
   Left_Right_Mask=NamedRerouteUsage()
  ) {materialFunction=/Game/Weapons/WeaponSkins/Wraps/MaterialFunctions/MF_Effect_Metal_LUT}
  n7 = NamedRerouteUsage()
  n8 = NamedRerouteUsage()
  n9 = NamedRerouteUsage()
  n10 = NamedRerouteUsage()
  n11 = NamedRerouteUsage()
  n12 = NamedRerouteUsage()
  n13 = NamedRerouteUsage()
  n14 = NamedRerouteUsage()
  n15 = NamedRerouteUsage()
  n16 = MaterialFunctionCall!(
   Material=MaterialFunctionCall!(
    Material=MaterialFunctionCall!(
     Material=@n6.Result_1
     IsState=@n7
     IsState_B=@n8
     Pre-Skinned Local Normal=@n9
     Calcated_Bounds=@n10
     Mask Primary=@n11
     Mask Secondary=@n12
     Object-SafeZone-Back=@n13
     Object-SafeZone-Middle=@n14
     Object-SafeZone-Front=@n15
    ) {materialFunction=/Game/Weapons/WeaponSkins/Wraps/MaterialFunctions/MF_Function_SDF_Blend}.Result
    Local Position=NamedRerouteUsage()
    Pre-Skinned Local Normal=NamedRerouteUsage()
    isState=NamedRerouteUsage()
    isVehicle=NamedRerouteUsage()
    "Customization Mask (_CM B)"=NamedRerouteUsage()
   ) {materialFunction=/Game/Weapons/WeaponSkins/Wraps/MaterialFunctions/MF_Sparkeles}.Result
   Decal_LR=NamedRerouteUsage()
   Decal_React=NamedRerouteUsage()
   IsState_B=NamedRerouteUsage()
   IsState=NamedRerouteUsage()
   Pre-Skinned Local Normal=NamedRerouteUsage()
   Calcated_Bounds=NamedRerouteUsage()
   Mask Primary=NamedRerouteUsage()
   Mask Secondary=NamedRerouteUsage()
   TS_Mask_02=NamedRerouteUsage()
   TopS Emissive_01=NamedRerouteUsage()
   TopS Emissive_02=NamedRerouteUsage()
   TS_Mask_01=NamedRerouteUsage()
   Object-SafeZone-VP_Charcter_Preview_Switch=NamedRerouteUsage()
   Object-SafeZone-Virtual-Pivot=NamedRerouteUsage()
  ) {materialFunction=/Game/Weapons/WeaponSkins/Wraps/MaterialFunctions/MF_Effect_HologramFX}
  n17 = NamedRerouteUsage()
  n18 = NamedRerouteUsage()
  n19 = NamedRerouteUsage()
  n20 = NamedRerouteUsage()
  n21 = NamedRerouteUsage()
  n22 = NamedRerouteUsage()
  n23 = NamedRerouteUsage()
  n24 = NamedRerouteUsage()
  n25 = MaterialFunctionCall!(
   Decal_Mask=NamedRerouteUsage()
   Material=MaterialFunctionCall!(
    Local Normal=NamedRerouteUsage()
    Local Position=NamedRerouteUsage()
    "Calculated Bounds "=NamedRerouteUsage()
    Is_States=NamedRerouteUsage()
    Material=MaterialFunctionCall!(
     Material=@n16.Result
     Decal_Mask=@n17
     Mask Crevise=@n18
     "Customization Mask (_CM B)"
     Mask Primary=@n19
     Mask Secondary=@n20
     TS_Mask_01=@n21
     TS_Mask_03=@n22
     TS_Mask_04=@n23
     TS_Mask_02=@n24
    ) {materialFunction=/Game/Weapons/WeaponSkins/Wraps/MaterialFunctions/MF_Effect_Iridescent}.after
   ) {materialFunction=/Game/Weapons/WeaponSkins/Wraps/MaterialFunctions/MF_Effect_SuperLevellingMask}.Result
   Mask_secondary=NamedRerouteUsage()
   Mask_Primary=NamedRerouteUsage()
   TS_Mask_01=NamedRerouteUsage()
   TS_Mask_02=NamedRerouteUsage()
  ) {materialFunction=/Game/Weapons/WeaponSkins/Wraps/MaterialFunctions/MF_Effect_ClothFuzz}
  n26 = NamedRerouteUsage()
  n27 = NamedRerouteUsage()
  n28 = NamedRerouteUsage()
  n29 = NamedRerouteUsage()
  n30 = NamedRerouteUsage()
  n31 = NamedRerouteUsage()
  n32 = NamedRerouteUsage()
  n33 = NamedRerouteUsage()
  n34 = NamedRerouteUsage()
  n35 = NamedRerouteUsage()
  n36 = MaterialFunctionCall!(
   Material=MaterialFunctionCall!(
    Material=MaterialFunctionCall!(
     Material=@n25.Result
     multi_Decal=@n26
     IsState=@n27
     CM_Msk=@n28
     IsState_B=@n29
     Pre-Skinned Local Normal
     Calcated_Bounds
     Decal_Mask=@n30
     Decal_Mask_RGBA=@n31
     TS_Layer04_Mask=@n32
     TS_Layer03_Mask=@n33
     TS_Layer01_Mask=@n34
     TS_Layer02_Mask=@n35
    ) {materialFunction=/Game/Weapons/WeaponSkins/Wraps/MaterialFunctions/MF_Effect_ThinFilm}.Result_1
    IsState=NamedRerouteUsage()
    IsState_B=NamedRerouteUsage()
    CM_Msk=NamedRerouteUsage()
    Pre-Skinned Local Normal
    Calcated_Bounds
    Decal_Mask_RGBA=NamedRerouteUsage()
    Decal_Mask=NamedRerouteUsage()
    TS_Layer02_Mask=NamedRerouteUsage()
    TS_Layer01_Mask=NamedRerouteUsage()
    TS_Layer03_Mask=NamedRerouteUsage()
    TS_Layer04_Mask=NamedRerouteUsage()
   ) {materialFunction=/Game/Weapons/WeaponSkins/Wraps/MaterialFunctions/MF_Effect_ClearCoat}.Result_1
   IsState_B=NamedRerouteUsage()
   IsState=NamedRerouteUsage()
   Pre-Skinned Local Normal=NamedRerouteUsage()
   Pre-Skinned Position=NamedRerouteUsage()
   Calcated_Bounds=NamedRerouteUsage()
   Mask Crevise=NamedRerouteUsage()
   Mask Secondary=NamedRerouteUsage()
   Mask Primary=NamedRerouteUsage()
   Bounds_Max=NamedRerouteUsage()
   Bounds_min=NamedRerouteUsage()
  ) {materialFunction=/Game/Weapons/WeaponSkins/Wraps/MaterialFunctions/MF_Effect_MuzzleGlow}
  n37 = NamedRerouteUsage()
  n38 = NamedRerouteUsage()
  n39 = NamedRerouteUsage()
  n40 = NamedRerouteUsage()
  n41 = NamedRerouteUsage()
  n42 = NamedRerouteUsage()
  n43 = NamedRerouteUsage()
  n44 = NamedRerouteUsage()
  n45 = NamedRerouteUsage()
  n46 = NamedRerouteUsage()
  n47 = NamedRerouteUsage()
  n48 = NamedRerouteUsage()
  n49 = MaterialFunctionCall!(
   Material=@n4.Result
   Material_Root=NamedRerouteUsage()
   isState=NamedRerouteUsage()
   isStateB=NamedRerouteUsage()
   Pre-Skinned Local Normal=NamedRerouteUsage()
   Calcated_Bounds=NamedRerouteUsage()
   Customization Mask Primary=NamedRerouteUsage()
   Customization Mask Secondary=NamedRerouteUsage()
   Mask Crevise=NamedRerouteUsage()
   Object-SafeZone-Back=NamedRerouteUsage()
   Object-SafeZone-Middle=NamedRerouteUsage()
   Object-SafeZone-Front=NamedRerouteUsage()
  ) {materialFunction=/Game/Weapons/WeaponSkins/Wraps/MaterialFunctions/MF_Function_Left_Right_Decal}
  n50 = MaterialFunctionCall(
   Material=MaterialFunctionCall(
    Material=@n49.Result_1
    Mask Crevise=NamedRerouteUsage()
    Mask Primary=NamedRerouteUsage()
    Mask Secondary=NamedRerouteUsage()
    TS_Mask_02=NamedRerouteUsage()
    TS_Mask_01=NamedRerouteUsage()
    Left_Right_Mask=@n49.Result_1
   ) {materialFunction=&MF_Effect_Metal_LUT}.Result_1
   IsState=NamedRerouteUsage()
   IsState_B=NamedRerouteUsage()
   Pre-Skinned Local Normal=NamedRerouteUsage()
   Calcated_Bounds=NamedRerouteUsage()
   Mask Primary=NamedRerouteUsage()
   Mask Secondary=NamedRerouteUsage()
   Object-SafeZone-Back=NamedRerouteUsage()
   Object-SafeZone-Middle=NamedRerouteUsage()
   Object-SafeZone-Front=NamedRerouteUsage()
  ) {materialFunction=&MF_Function_SDF_Blend}
  n51 = NamedRerouteUsage()
  n52 = NamedRerouteUsage()
  n53 = NamedRerouteUsage()
  n54 = NamedRerouteUsage()
  n55 = NamedRerouteUsage()
  n56 = MaterialFunctionCall!(
   Material=MaterialFunctionCall!(
    ForwardAxis=NamedRerouteUsage()
    Material=MaterialFunctionCall(Material=@n50.Result, Local Position=@n51, Pre-Skinned Local Normal=@n52, isState=@n53, isVehicle=@n54, "Customization Mask (_CM B)"=@n55) {materialFunction=&MF_Sparkeles}.Result
    isMech=NamedRerouteUsage()
    isCharacter=NamedRerouteUsage()
    isVehicle=NamedRerouteUsage()
    Pre-Skinned Local Normal=NamedRerouteUsage()
    Local Position=NamedRerouteUsage()
    Object Bounds Min Adjustment=NamedRerouteUsage()
    Object Bounds Max Adjustment=NamedRerouteUsage()
    Object Bounds Center Adjustment=NamedRerouteUsage()
    "Customization Mask (_CM B Top left)"=NamedRerouteUsage()
    "Customization Mask 2 (_CM B Top Right)"=NamedRerouteUsage()
   ) {materialFunction=/Game/Weapons/WeaponSkins/Wraps/MaterialFunctions/MF_ElectricBurst_Wrap}.Result
   Local Bounds=NamedRerouteUsage()
   isVehicle=NamedRerouteUsage()
   isMech=NamedRerouteUsage()
   Object Bounds Min Adjustment=NamedRerouteUsage()
   Object Bounds Max Adjustment=NamedRerouteUsage()
   Object Bounds Center Adjustment=NamedRerouteUsage()
   "Customization Mask 1 (_CM B Top Left)"=NamedRerouteUsage()
   Pre-Skinned Local Position=NamedRerouteUsage()
   Pre-Skinned Local Normal=NamedRerouteUsage()
  ) {materialFunction=/Game/Weapons/WeaponSkins/Wraps/MaterialFunctions/MF_SignalInterference_Wrap}
  n57 = NamedRerouteUsage()
  n58 = NamedRerouteUsage()
  n59 = Max(A=NamedRerouteUsage(), B=NamedRerouteUsage())
  n60 = MaterialFunctionCall(
   Material=MaterialFunctionCall!(
    Customization Mask
    Local Position=NamedRerouteUsage()
    Local Bounds=NamedRerouteUsage()
    isVehicle=NamedRerouteUsage()
    Local Normal=NamedRerouteUsage()
    Material=MaterialFunctionCall!(Material=@n56.Result, "Customization Crevasse (_CM G)"=@n57, Secondary Color=@n58, "Customization Mask (_CM B)"=@n59) {materialFunction=/Game/Weapons/WeaponSkins/Wraps/MaterialFunctions/MF_Cartoon}.Result
   ) {materialFunction=/Game/Weapons/WeaponSkins/Wraps/MaterialFunctions/MF_MilitaryMan}.Result
   IsState_B=NamedRerouteUsage()
   IsState=NamedRerouteUsage()
   Pre-Skinned Local Normal=NamedRerouteUsage()
   Calcated_Bounds=NamedRerouteUsage()
   Mask Primary=NamedRerouteUsage()
   Mask Secondary=NamedRerouteUsage()
   TS_Mask_02=NamedRerouteUsage()
   TopS Emissive_01=NamedRerouteUsage()
   TopS Emissive_02=NamedRerouteUsage()
   TS_Mask_01=NamedRerouteUsage()
   Object-SafeZone-VP_Charcter_Preview_Switch=NamedRerouteUsage()
   Object-SafeZone-Virtual-Pivot=NamedRerouteUsage()
  ) {materialFunction=&MF_Effect_HologramFX}
  n61 = StaticSwitch(
   True=@n60.Result
   False=StaticSwitch(
    True=NamedRerouteUsage()
    False=@n60.Result
    Value=StaticBoolParameter() {group=Customization, parameterName=Custmization One_Offs, sortPriority=-500}
   )
   Value=NamedRerouteUsage()
  )
  n62 = NamedRerouteUsage()
  n63 = NamedRerouteUsage()
  n64 = NamedRerouteUsage()
  n65 = NamedRerouteUsage()
  n66 = NamedRerouteUsage()
  n67 = NamedRerouteUsage()
  n68 = NamedRerouteUsage()
  n69 = NamedRerouteUsage()
  n70 = NamedRerouteUsage()
  n71 = NamedRerouteUsage()
  n72 = NamedRerouteUsage()
  n73 = NamedRerouteUsage()
  n74 = MaterialFunctionCall!(
   ObjectBounds
   Material Attributes=MaterialFunctionCall!(
    Material=MaterialFunctionCall!(
     Material=@n61
     ForwardAxis=@n62
     isMech=@n63
     Wrap ScaleMultiplier=@n64
     PanDirection=@n62
     isVehicle=@n65
     IsCharacter=@n66
     IsStatsB=@n67
     "Customization Mask (_CM B)"=@n68
     Pre-Skinned Local Position=@n69
     Pre-Skinned Local Normal=@n70
     Object Bounds Center Adjustment=@n71
     Object Bounds Min Adjustment=@n72
     Object Bounds Max Adjustment=@n73
    ) {materialFunction=/Game/Weapons/WeaponSkins/Wraps/MaterialFunctions/MF_Customization_Drippy}.Result
    IsStatsB=NamedRerouteUsage()
    IsCharacter=NamedRerouteUsage()
    Local Bounds=NamedRerouteUsage()
    isVehicle=NamedRerouteUsage()
    isMech=NamedRerouteUsage()
    "Customization Mask 1 (_CM B Top Left)"=NamedRerouteUsage()
    "Customization Mask 2 (_CM B Top Right)"=NamedRerouteUsage()
    Pre-Skinned Local Position=NamedRerouteUsage()
    Pre-Skinned Local Normal=NamedRerouteUsage()
   ) {materialFunction=/Game/Weapons/WeaponSkins/Wraps/MaterialFunctions/MF_Customization_SwipingPatterns}.Result
   isVehicle=NamedRerouteUsage()
   WorldPosition
   IsCharacter=NamedRerouteUsage()
   SideProjectVehicle=NamedRerouteUsage()
   IsStatsB=NamedRerouteUsage()
   Normals
   "Customization Mask 1 (_CM B Top Left)"=NamedRerouteUsage()
   "Customization Mask 2 (_CM B Top Right)"=NamedRerouteUsage()
  ) {materialFunction=/Game/Weapons/WeaponSkins/Wraps/MaterialFunctions/MF_Yatter}
  n75 = NamedRerouteUsage()
  n76 = NamedRerouteUsage()
  n77 = NamedRerouteUsage()
  n78 = NamedRerouteUsage()
  n79 = NamedRerouteUsage()
  n80 = MaterialFunctionCall!(
   Material=MaterialFunctionCall!(
    CM1=NamedRerouteUsage()
    CM2=NamedRerouteUsage()
    Material=MaterialFunctionCall!(Material=@n74.Result, IsVehicle=@n75, IsMech=@n76, Bounds=@n77, Side Project Vehicle=@n78, PreskinnedLocalPosition=@n79) {materialFunction=/Game/Weapons/WeaponSkins/Wraps/MaterialFunctions/MF_GlitchyFX}.Result
   ) {materialFunction=/Game/Weapons/WeaponSkins/Wraps/MaterialFunctions/MF_Columbus}.Result
   IsCharacter=NamedRerouteUsage()
   Local Bounds=NamedRerouteUsage()
   isMech=NamedRerouteUsage()
   isVehicle=NamedRerouteUsage()
   IsStatsB=NamedRerouteUsage()
   "Customization Mask 1 (_CM B Top Left)"=NamedRerouteUsage()
   "Customization Mask 2 (_CM B Top Right)"=NamedRerouteUsage()
   Pre-Skinned Local Position=NamedRerouteUsage()
   Pre-Skinned Local Normal=NamedRerouteUsage()
  ) {materialFunction=/Game/Weapons/WeaponSkins/Wraps/MaterialFunctions/MF_Customization_TriggeredPatterns}
  n81 = NamedRerouteUsage()
  n82 = NamedRerouteUsage()
  n83 = NamedRerouteUsage()
  n84 = NamedRerouteUsage()
  n85 = NamedRerouteUsage()
  n86 = NamedRerouteUsage()
  n87 = NamedRerouteUsage()
  n88 = NamedRerouteUsage()
  n89 = MaterialFunctionCall!(
   GMap=NamedRerouteUsage()
   TopLayerMaterial_Custom
   BottomLayerMaterial_Custom
   Crevasse=NamedRerouteUsage()
   isMech=NamedRerouteUsage()
   isVehicle=NamedRerouteUsage()
   Material=MaterialFunctionCall!(
    Pre-Skinned World Normal=NamedRerouteUsage()
    Gmap=NamedRerouteUsage()
    Customization Mask=NamedRerouteUsage()
    IsCharacter=NamedRerouteUsage()
    IsStatsB=NamedRerouteUsage()
    Material Attributes=MaterialFunctionCall!(
     Material=@n80.Result
     isMech=@n81
     Local Bounds=@n82
     isVehicle=@n83
     GMap=@n84
     "Customization Mask 1 (_CM B Top Left)"=@n85
     "Customization Mask 2 (_CM B Top Right)"=@n86
     Pre-Skinned Local Position=@n87
     Pre-Skinned Local Normal=@n88
    ) {materialFunction=/Game/Weapons/WeaponSkins/Wraps/MaterialFunctions/MF_TNTina_Wrap}.Result
   ) {materialFunction=/Game/Weapons/WeaponSkins/MaterialFunction/Weapon_SkinFXFunction}.Wrap_OutPut
   PreskinnedPosition=NamedRerouteUsage()
   PreskinnedNormal=NamedRerouteUsage()
   Root_Material=NamedRerouteUsage()
   "Customization Mask (_CM B Top left)"=NamedRerouteUsage()
   "Customization Mask 2 (_CM B Top Right)"=NamedRerouteUsage()
  ) {materialFunction=/Game/Weapons/WeaponSkins/Wraps/MaterialFunctions/MF_DirtAndEdgeWearMasked}
  n90 = NamedRerouteUsage()
  n91 = NamedRerouteUsage()
  n92 = NamedRerouteUsage()
  n93 = NamedRerouteUsage()
  n94 = NamedRerouteUsage()
  n95 = NamedRerouteUsage()
  n96 = NamedRerouteUsage()
  n97 = NamedRerouteUsage()
  n98 = NamedRerouteUsage()
  n99 = NamedRerouteUsage()
  n100 = MaterialFunctionCall(
   Material=MaterialFunctionCall!(
    IsStatsB=NamedRerouteUsage()
    Material=MaterialFunctionCall!(
     Side Project Vehicle=@n48
     Material=@n89.Result
     ForwardAxis=@n90
     IsCharacter=@n91
     isMech=@n92
     isVehicle=@n93
     Pre-Skinned Local Normal=@n94
     Local Position=@n95
     Wrap Bounds=@n96
     "Customization Mask (_CM B Top left)"=@n97
     "Customization Mask 2 (_CM B Top Right)"=@n98
     isCharAccent=@n99
    ) {materialFunction=/Game/Weapons/WeaponSkins/Wraps/MaterialFunctions/MF_SpinningStars_Wrap}.Result
    GalaxyAlternateFresnel
    ForwardAxis=NamedRerouteUsage()
    Warp Mask
    Custom MAsk=NamedRerouteUsage()
    Side Project Vehicle=NamedRerouteUsage()
    isMech=NamedRerouteUsage()
    isCharacter=NamedRerouteUsage()
    isVehicle=NamedRerouteUsage()
    Calcated_Bounds=NamedRerouteUsage()
   ) {materialFunction=/Game/Weapons/WeaponSkins/Wraps/MaterialFunctions/MF_Wrap_Galaxy}.Result_1
   Decal_Mask=NamedRerouteUsage()
   Mask Crevise=NamedRerouteUsage()
   Mask Primary=NamedRerouteUsage()
   Mask Secondary=NamedRerouteUsage()
   TS_Mask_01=NamedRerouteUsage()
   TS_Mask_02=NamedRerouteUsage()
  ) {materialFunction=&MF_Effect_Iridescent}
  n101 = NamedRerouteUsage()
  n102 = NamedRerouteUsage()
  n103 = NamedRerouteUsage()
  n104 = MaterialFunctionCall(
   Decal_Mask=NamedRerouteUsage()
   Material=MaterialFunctionCall(
    Local Normal=NamedRerouteUsage()
    Local Position=NamedRerouteUsage()
    "Calculated Bounds "=NamedRerouteUsage()
    Is_States=NamedRerouteUsage()
    Material=MaterialFunctionCall!(
     Local Normal=@n46
     Local Position_1=@n47
     In=@n100.after
     isVehicle_1=@n101
     "Customization Mask (_CM B Top left)"=@n102
     "Customization Mask 2 (_CM B Top Right)"=@n103
    ) {materialFunction=/Game/Weapons/WeaponSkins/Wraps/MaterialFunctions/MF_FlutterNormals}.Result
   ) {materialFunction=&MF_Effect_SuperLevellingMask}.Result
   Mask_secondary=NamedRerouteUsage()
   Mask_Primary=NamedRerouteUsage()
   TS_Mask_01=NamedRerouteUsage()
   TS_Mask_02=NamedRerouteUsage()
  ) {materialFunction=&MF_Effect_ClothFuzz}
  n105 = NamedRerouteUsage()
  n106 = NamedRerouteUsage()
  n107 = NamedRerouteUsage()
  n108 = NamedRerouteUsage()
  n109 = NamedRerouteUsage()
  n110 = NamedRerouteUsage()
  n111 = NamedRerouteUsage()
  n112 = NamedRerouteUsage()
  n113 = NamedRerouteUsage()
  n114 = NamedRerouteUsage()
  n115 = MaterialFunctionCall!(
   Material=MaterialFunctionCall!(
    MA_In=MaterialFunctionCall(
     Material=@n104.Result
     IsState_B=@n105
     IsState=@n106
     Pre-Skinned Local Normal=@n107
     Pre-Skinned Position=@n108
     Calcated_Bounds=@n109
     Mask Crevise=@n110
     Mask Secondary=@n111
     Mask Primary=@n112
     Bounds_Max=@n113
     Bounds_min=@n114
    ) {materialFunction=&MF_Effect_MuzzleGlow}.Result
    CustomizationMask=NamedRerouteUsage()
   ) {materialFunction=/Game/Weapons/WeaponSkins/Wraps/MaterialFunctions/MF_DigitalRain}.Result
   Axis
   IsStates=NamedRerouteUsage()
   IsStates_B=NamedRerouteUsage()
   Pre-Skinned Local Normal=NamedRerouteUsage()
   Calcated_Bounds=NamedRerouteUsage()
   Object Bounds Min Adjustment
   Object Bounds Max Adjustment
   Object Bounds Center Adjustment
   "Customization Mask (_CM B)Right"=NamedRerouteUsage()
   "Customization Mask (_CM B)left"=NamedRerouteUsage()
   Object-SafeZone-Back=NamedRerouteUsage()
   Object-SafeZone-Middle=NamedRerouteUsage()
   Object-SafeZone-Front=NamedRerouteUsage()
   Object-SafeZone-VP-Character-PreView-Switch=NamedRerouteUsage()
   Object-SafeZone-Virtual-Pivot=NamedRerouteUsage()
  ) {materialFunction=/Game/Weapons/WeaponSkins/Wraps/MaterialFunctions/MF_Uvchecker}
  n116 = NamedRerouteUsage()
  n117 = MaterialFunctionCall!(
   Root_Material=NamedRerouteUsage()
   Material=StaticSwitch(
    True=MaterialFunctionCall!(
     Material Root=@n1
     "Msk_Ts01<--->Ts02"=@n2
     Material In=@n36.Result
     HotWax_Msk=@n37
     CM_Msk=@n38
     IsState_B=@n39
     Decal_Mask=@n40
     IsState=@n41
     Pre-Skinned Local Normal=@n42
     Calcated_Bounds=@n43
     TopsSide_Msk 01=@n44
     TopsSide_Msk 02=@n45
    ) {materialFunction=/Game/Weapons/WeaponSkins/Wraps/MaterialFunctions/MF_Effect_EdgeWear_Dirt_V2}.Result_1
    False=MaterialFunctionCall!(MA=@n115.Result, CustomizationMask=@n116) {materialFunction=/Game/Weapons/WeaponSkins/Wraps/MaterialFunctions/MF_SimpleLCDScreen}.Result
    Value=NamedRerouteUsage()
   )
   ForwardAxis=NamedRerouteUsage()
   Hotwax_Msk=NamedRerouteUsage()
   IsState_BV4=NamedRerouteUsage()
   CM_Msk=NamedRerouteUsage()
   IsState=NamedRerouteUsage()
   IsState_B=NamedRerouteUsage()
   Pre-Skinned Local Normal=NamedRerouteUsage()
   Calcated_Bounds=NamedRerouteUsage()
   Cropped Emissive Mask=NamedRerouteUsage()
   Per_ObjectMaskOverride=NamedRerouteUsage()
   Object-SafeZone-Back=NamedRerouteUsage()
   Object-SafeZone-Middle=NamedRerouteUsage()
   Object-SafeZone-Front=NamedRerouteUsage()
  ) {
   desc="all MF's that should be last "
   materialFunction=&/MF_Base_Customization_Effects
  }
  n118 = NamedRerouteUsage()
  n119 = MaterialFunctionCall!(
   Original Material=@n118
   Wrapped Material=@n117.Result
   Enable=FunctionInput(Preview=StaticBool()) {bUsePreviewValueAsDefault=true, inputName=Wrap Holdout, inputType=FunctionInput_StaticBool, sortPriority=110}
  ) {
   desc="Prevents wrap from applying\r\nin specified areas\r\nuses Alpha vertex color"
   materialFunction=&/MF_WrapHoldout
  }
  FunctionOutput(None=StaticSwitchParameter(True=@n119.Result, False=@n118) {defaultValue=true, group=Customization, parameterName=Allow Customization Effects, sortPriority=-500})
  MaterialFunctionCall!(
   Material_Root=NamedRerouteUsage()
   Material=@n5
   "Msk_Ts01<--->Ts02"
   ForwardAxis=NamedRerouteUsage()
   IsState=NamedRerouteUsage()
   CM_Msk=NamedRerouteUsage()
   IsState_BV4=NamedRerouteUsage()
   Pre-Skinned Local Normal=NamedRerouteUsage()
   Calcated_Bounds=NamedRerouteUsage()
   Object-SafeZone-Back=NamedRerouteUsage()
   Object-SafeZone-Middle=NamedRerouteUsage()
   Object-SafeZone-Front=NamedRerouteUsage()
  ) {materialFunction=/Game/Weapons/WeaponSkins/Wraps/MaterialFunctions/MF_Function_Left_Right_Decal_V2}
  NamedRerouteDeclaration(None=FunctionInput() {bUsePreviewValueAsDefault=true, inputName=Object-SafeZone-Back, inputType=FunctionInput_Vector4, sortPriority=100}) {name=SafeZone-Back, nodeColor=<rgba 1 0.6588 0.0581 1>}
  NamedRerouteDeclaration(None=FunctionInput() {bUsePreviewValueAsDefault=true, inputName=Object-SafeZone-Middle, inputType=FunctionInput_Vector4, sortPriority=101}) {name=SafeZone-Middle, nodeColor=<rgba 1 0.6588 0.0581 1>}
  NamedRerouteDeclaration(None=FunctionInput() {bUsePreviewValueAsDefault=true, inputName="Customization Mask (_CM RGBA)", inputType=FunctionInput_Vector4, sortPriority=26}) {name=Masks_RGBA_CM, nodeColor=<rgba 1 0.1875 0 1>}
  n120 = NamedRerouteDeclaration(None=FunctionInput() {bUsePreviewValueAsDefault=true, inputName=isBackpack, inputType=FunctionInput_Scalar, sortPriority=78}) {name=IsBackpack, nodeColor=<rgba 0 0.9253 1 1>}
  n121 = NamedRerouteDeclaration(None=FunctionInput() {bUsePreviewValueAsDefault=true, inputName=isPickaxe, inputType=FunctionInput_Scalar, sortPriority=78}) {name=IsPickaxe, nodeColor=<rgba 0 0.9253 1 1>}
  NamedRerouteDeclaration(None=MaterialFunctionCall(X=@n120, Y=@n121) {materialFunction=&MakeFloat2}.Result) {name=Is_StatsB, nodeColor=<rgba 0 0.9253 1 1>}
  n122 = FunctionInput() {bUsePreviewValueAsDefault=true, inputName=Object-SafeZone-VP_Charcter_Preview_Switch, inputType=FunctionInput_Scalar, sortPriority=105}
  NamedRerouteDeclaration(None=@n122) {name=SafeZone-VP_Charcter_Preview_Switch, nodeColor=<rgba 1 0.6588 0.0581 1>}
  NamedRerouteDeclaration(None=FunctionInput() {bUsePreviewValueAsDefault=true, inputName=Pre-Skinned Local Normal, sortPriority=42}) {name=PreSkin_Normals, nodeColor=<rgba 1 0 0.5156 1>}
  n123 = MaterialFunctionCall!(
   Material Root
   Material In=NamedRerouteUsage()
   ------------------
   IsStateS=NamedRerouteUsage()
   IsState_B=NamedRerouteUsage()
   IsState_B_V4=NamedRerouteUsage()
   Forward Axis=NamedRerouteUsage()
   Rotation Z Axis=NamedRerouteUsage()
   Wrap ScaleMultiplier=NamedRerouteUsage()
   Rotation X Axis=NamedRerouteUsage()
   Caculated_Bounds=NamedRerouteUsage()
   "Customization Mask (_CM RGBA)"=NamedRerouteUsage()
   --------------------
   Pre-Skinned Local Position=NamedRerouteUsage()
   Pre-Skinned Local Normal=NamedRerouteUsage()
   Pre-Skinned Bounds=NamedRerouteUsage()
   Object-SafeZone-Back=NamedRerouteUsage()
   Object-SafeZone-Middle=NamedRerouteUsage()
   Object-SafeZone-Front=NamedRerouteUsage()
   Object-SafeZone-Virtual-Pivot=NamedRerouteUsage()
   Object-SafeZone-VP_Charcter_Preview_Switch=NamedRerouteUsage()
  ) {materialFunction=&/MF_Base_Customization_Primary_V2}
  n124 = StaticBoolParameter() {group=Customization, parameterName="Custmization  V2", sortPriority=-500}
  NamedRerouteDeclaration(None=StaticSwitch(True=@n123.TS_Mask_02, False=@n3.TS_Mask_02, Value=@n124)) {name=TS_Mask_02, nodeColor=<rgba 0 0.9062 1 1>}
  NamedRerouteDeclaration(None=StaticSwitch(True=@n123.TS_Mask_01, False=@n3.TS_Mask_01, Value=@n124)) {name=TS_Mask_01, nodeColor=<rgba 0 0.9062 1 1>}
  NamedRerouteDeclaration(None=FunctionInput() {bUsePreviewValueAsDefault=true, inputName=Object-SafeZone-Front, inputType=FunctionInput_Vector4, sortPriority=102}) {name=SafeZone-Front, nodeColor=<rgba 1 0.6588 0.0581 1>}
  NamedRerouteDeclaration(None=FunctionInput() {inputName=Material, inputType=FunctionInput_MaterialAttributes, sortPriority=-200}) {name=Root_Mat, nodeColor=<rgba 0.8047 1 0 1>}
  NamedRerouteDeclaration(None=@n123.TS_Mask_03) {name=TS_Mask_03, nodeColor=<rgba 1 0 0.7031 1>}
  NamedRerouteDeclaration(None=FunctionInput() {bUsePreviewValueAsDefault=true, inputName=Pre-Skinned Local Position, sortPriority=41}) {name=PreSkin_Position, nodeColor=<rgba 1 0 0.5156 1>}
  NamedRerouteDeclaration(None=FunctionInput() {bUsePreviewValueAsDefault=true, inputName=Object Bounds Max Adjustment}) {name=Bounds_Max, nodeColor=<rgba 0.5234 1 0 1>}
  NamedRerouteDeclaration(None=FunctionInput() {bUsePreviewValueAsDefault=true, inputName=Object Bounds Min Adjustment}) {name=Bounds_Min, nodeColor=<rgba 0.5234 1 0 1>}
  NamedRerouteDeclaration(None=FunctionInput(Preview=Constant() {r=1}) {bUsePreviewValueAsDefault=true, inputName=Per_ObjectMaskOverride, inputType=FunctionInput_Scalar, sortPriority=75}) {name=PerObjectMask, nodeColor=<rgba 1 0.6569 0.0791 1>}
  n125 = FunctionInput() {bUsePreviewValueAsDefault=true, inputName=isCharAccent, inputType=FunctionInput_Scalar, sortPriority=78}
  NamedRerouteDeclaration(None=MaterialFunctionCall(X=@n120, Y=@n121, Z=Constant(), A=NamedRerouteDeclaration(None=@n125) {name=isCharAccent, nodeColor=<rgba 0 0.9253 1 1>}) {materialFunction=&MakeFloat4}.Result) {name=Is_StatsB_V4, nodeColor=<rgba 0 0.9253 1 1>}
  NamedRerouteDeclaration(None=FunctionInput() {bUsePreviewValueAsDefault=true, inputName=Pre-Skinned Bounds, sortPriority=41}) {name=PreSkin_Bounds, nodeColor=<rgba 1 0.4219 0 1>}
  n126 = FunctionInput() {bUsePreviewValueAsDefault=true, inputName=Object-SafeZone-Virtual-Pivot, inputType=FunctionInput_Vector4, sortPriority=105}
  NamedRerouteDeclaration(None=@n126) {name=SafeZone-Virtual-Pivot, nodeColor=<rgba 1 0.6588 0.0581 1>}
  NamedRerouteDeclaration(None=StaticSwitch(True=@n123.TS_Emis_01_Msk, False=@n3.TS_Emis_01_Msk, Value=@n124)) {name=TS_Emis_01_Msk, nodeColor=<rgba 1 0.8672 0 1>}
  NamedRerouteDeclaration(None=StaticSwitch(True=@n123.TS_Emis_02_Msk, False=@n3.TS_Emis_02_Msk, Value=@n124)) {name=TS_Emis_02_Msk, nodeColor=<rgba 1 0.8672 0 1>}
  NamedRerouteDeclaration(None=FunctionInput() {bUsePreviewValueAsDefault=true, inputName=Object Bounds Center Adjustment}) {name=Bounds_Center, nodeColor=<rgba 0.5234 1 0 1>}
  NamedRerouteDeclaration(None=FunctionInput() {bUsePreviewValueAsDefault=true, inputName=Rotation Z Axis, inputType=FunctionInput_Scalar}) {name=Rotation_Z, nodeColor=<rgba 0.1016 0 1 1>}
  NamedRerouteDeclaration(None=FunctionInput() {bUsePreviewValueAsDefault=true, inputName=Rotation X Axis, inputType=FunctionInput_Scalar}) {name=Rotation_X, nodeColor=<rgba 0.1016 0 1 1>}
  NamedRerouteDeclaration(None=FunctionInput() {bUsePreviewValueAsDefault=true, inputName=Forward Axis}) {name=Forward Axis, nodeColor=<rgba 1 0.1875 0 1>}
  n127 = FunctionInput() {bUsePreviewValueAsDefault=true, inputName=Wrap ScaleMultiplier, inputType=FunctionInput_Scalar, previewValue=<1 0 0 1>}
  NamedRerouteDeclaration(None=@n127) {name=ScaleMultiplier, nodeColor=<rgba 0.9922 1 0 1>}
  NamedRerouteDeclaration(None=FunctionInput() {bUsePreviewValueAsDefault=true, inputName=Cropped Emissive Mask, sortPriority=20}) {name=Cropped Emissive Mask, nodeColor=<rgba 1 0.0938 0 1>}
  NamedRerouteDeclaration(None=FunctionInput() {bUsePreviewValueAsDefault=true, inputName=Normal Map, sortPriority=15}) {name=Normal_Map, nodeColor=<rgba 0 0.1094 1 1>}
  n128 = FunctionInput() {
   bUsePreviewValueAsDefault=true
   inputName="Customization Mask UV Islands (_CM B Bottom Right)"
   inputType=FunctionInput_Scalar
   sortPriority=22
  }
  NamedRerouteDeclaration(None=@n128) {name=Mask_UV_islands, nodeColor=<rgba 0.1953 0 1 1>}
  NamedRerouteDeclaration(None=FunctionInput() {bUsePreviewValueAsDefault=true, inputName="Customization Mask(_CM A)", inputType=FunctionInput_Scalar}) {name=Mask_Gmap, nodeColor=<rgba 0.0094 0.0094 0.901 1>}
  n129 = NamedRerouteUsage()
  n130 = NamedRerouteUsage()
  n131 = NamedRerouteUsage()
  n132 = NamedRerouteUsage()
  n133 = NamedRerouteUsage()
  n134 = NamedRerouteUsage()
  n135 = NamedRerouteUsage()
  n136 = NamedRerouteUsage()
  n137 = NamedRerouteUsage()
  n138 = NamedRerouteUsage()
  n139 = NamedRerouteUsage()
  n140 = NamedRerouteUsage()
  n141 = NamedRerouteUsage()
  n142 = NamedRerouteUsage()
  n143 = NamedRerouteUsage()
  n144 = NamedRerouteUsage()
  n145 = NamedRerouteUsage()
  n146 = NamedRerouteUsage()
  n147 = NamedRerouteUsage()
  n148 = NamedRerouteUsage()
  n149 = NamedRerouteUsage()
  n150 = NamedRerouteUsage()
  NamedRerouteDeclaration(None=MaterialFunctionCall(
    Material=MaterialFunctionCall!(
     Material=@n129
     ------------------
     IsState_B_V4=@n130
     IsStateS=@n131
     IsState_B=@n132
     Rotation Z Axis=@n133
     Forward Axis=@n134
     Rotation X Axis=@n135
     Wrap ScaleMultiplier=@n136
     Object Bounds Center Adjustment=@n137
     Object Bounds Max Adjustment=@n138
     Object SBounds Min Adjustment=@n139
     Bounds_Extent=@n140
     Caculated_Bounds=@n141
     "Customization Mask (_CM RGBA)"=@n142
     --------------------
     Pre-Skinned Local Position=@n143
     Pre-Skinned Local Normal=@n144
     Pre-Skinned Bounds=@n145
     Object-SafeZone-Back=@n146
     Object-SafeZone-Middle=@n147
     Object-SafeZone-Front=@n148
     Object-SafeZone-Virtual-Pivot=@n149
     Object-SafeZone-VP_Charcter_Preview_Switch=@n150
    ) {materialFunction=&/MF_Base_One_Offs}.Result
    IsVehicle=NamedRerouteUsage()
    is_StateB=NamedRerouteUsage()
    Normal Map=NamedRerouteUsage()
    "Customization Crevasse (_CM G)"=NamedRerouteUsage()
    "Customization Mask UV Islands (_CM B Bottom Right)"=NamedRerouteUsage()
    "Customization Mask (_CM B)"=NamedRerouteUsage()
    Pre-Skinned Local Position=NamedRerouteUsage()
    Pre-Skinned Local Normal=NamedRerouteUsage()
    isCharAccent=NamedRerouteUsage()
   ) {materialFunction=&MF_Customization_Simple}.Result) {name=MF_Base_One_Offs, nodeColor=<rgba 1 0 0.6562 1>}
  n151 = MaterialFunctionCall!(
   is Vehicle=NamedRerouteUsage()
   Object Bounds Max Adjustment=NamedRerouteUsage()
   Object Bounds Min Adjustment=NamedRerouteUsage()
   Object Bounds Center Adjustment=NamedRerouteUsage()
   isMech=NamedRerouteUsage()
   Side Project Vehicle=NamedRerouteUsage()
   Local Position=NamedRerouteUsage()
  ) {materialFunction=/Game/Weapons/WeaponSkins/Wraps/MaterialFunctions/MF_WrapBounds}
  NamedRerouteDeclaration(None=@n151.Bound Extents) {name=Bounds Extents, nodeColor=<rgba 1 0.6797 0 1>}
  NamedRerouteDeclaration(None=@n124) {name="Custmization  V2 State", nodeColor=<rgba 1 0.1172 0 1>}
  n152 = FunctionInput() {bUsePreviewValueAsDefault=true, inputName="Customization Mask 1 (_CM B Top Left)", inputType=FunctionInput_Scalar, sortPriority=24}
  n153 = NamedRerouteDeclaration(None=@n152) {name="Mask_Primary {B Top Left}", nodeColor=<rgba 0.0094 0.0094 0.901 1>}
  n154 = FunctionInput() {bUsePreviewValueAsDefault=true, inputName="Customization Mask 2 (_CM B Top Right)", inputType=FunctionInput_Scalar, sortPriority=25}
  n155 = NamedRerouteDeclaration(None=@n154) {name="Mask_Secondary {B Top Right}", nodeColor=<rgba 0.0094 0.0094 0.901 1>}
  n156 = FunctionInput(Preview=TextureSample(UVs=TextureCoordinate()) {samplerType=SAMPLERTYPE_Masks, texture=/Game/Weapons/FORT_Rifles/Textures/SCAR_M}.G) {bUsePreviewValueAsDefault=true, inputName="Customization Crevasse (_CM G)", inputType=FunctionInput_Scalar, sortPriority=21}
  n157 = Add(A=@n153, B=@n155)
  NamedRerouteDeclaration(None=MaterialFunctionCall(X=@n153, Y=@n155, Z=NamedRerouteDeclaration(None=@n156) {name=Crevice_Mask, nodeColor=<rgba 0 0.7865 0 1>}, A=Saturate(None=@n157)) {materialFunction=&MakeFloat4}.Result) {name=Mask_Custom, nodeColor=<rgba 0.4297 0 1 1>}
  NamedRerouteDeclaration(None=@n123.Result) {name=Custmization V2, nodeColor=<rgba 1 0 0.0234 1>}
  NamedRerouteDeclaration(None=@n123.Decal_Mask_Alpha) {name=Decal_Mask_V2_Alpha, nodeColor=<rgba 1 0 0.0469 1>}
  NamedRerouteDeclaration(None=@n123.TS_Mask_04) {name=TS_Mask_04, nodeColor=<rgba 1 0 0.7031 1>}
  NamedRerouteDeclaration(None=@n49.Decal_Mask) {name=Decal_Msk, nodeColor=<rgba 0 0.4141 1 1>}
  NamedRerouteDeclaration(None=@n123.Decal_RGBA) {name=Decal_Mask_V2_RGBA, nodeColor=<rgba 1 0 0.0469 1>}
  NamedRerouteDeclaration(None=@n123.Hotwax_Msk) {name=HotWax_Msk, nodeColor=<rgba 0.2656 1 0 1>}
  NamedRerouteDeclaration(None=@n123."MSk_TS1<->Ts2") {name="MSk_TS1<->Ts2", nodeColor=<rgba 0 0.9531 1 1>}
  NamedRerouteDeclaration(None=@n123.Decal_Multi_Msk) {name=Decal_Multi_Msk, nodeColor=<rgba 0 0.0625 1 1>}
  n158 = FunctionInput() {bUsePreviewValueAsDefault=true, inputName=isVehicle, inputType=FunctionInput_Scalar, sortPriority=75}
  n159 = FunctionInput() {bUsePreviewValueAsDefault=true, inputName=Side Project Vehicle, inputType=FunctionInput_Scalar, sortPriority=76}
  n160 = FunctionInput() {bUsePreviewValueAsDefault=true, inputName=isMech, inputType=FunctionInput_Scalar, sortPriority=77}
  n161 = FunctionInput() {bUsePreviewValueAsDefault=true, inputName=isCharacter, inputType=FunctionInput_Scalar, sortPriority=78}
  NamedRerouteDeclaration(None=MaterialFunctionCall(
    X=NamedRerouteDeclaration(None=@n158) {name=isVehicle, nodeColor=<rgba 0 0.9253 1 1>}
    Y=NamedRerouteDeclaration(None=@n159) {name=isVehicleSide, nodeColor=<rgba 0 0.9253 1 1>}
    Z=NamedRerouteDeclaration(None=@n160) {name=isMech, nodeColor=<rgba 0 0.9253 1 1>}
    A=NamedRerouteDeclaration(None=@n161) {name=isCharacter, nodeColor=<rgba 0 0.9253 1 1>}
   ) {materialFunction=&MakeFloat4}.Result) {name=Is_States, desc="R=vehicle G=VehicleSide =R=Mech A=Character", nodeColor=<rgba 0 0.9253 1 1>}
  NamedRerouteDeclaration(None=@n151.Result) {name="Calculated Bounds ", nodeColor=<rgba 1 0.6797 0 1>}
  NamedRerouteUsage()
  NamedRerouteUsage()
  NamedRerouteUsage()
  NamedRerouteUsage()
 ]}
