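# Placeholder master material function that outputs minimal Material Attributes plus passthrough data for customization.
# Material Result sets only emissive, from MF_Base_CroppedEmissive, and normal, from a flat normal texture, leaving other attributes unconnected.
# Extra outputs expose that normal, the cropped emissive mask, MF_Base_CustomizationMask channels (masks, crevasse, UV islands, gmap) and MF_Base_LocalPosition pre-skinned position, normal and bounds.
# Dashed constant outputs serve as dividers between output groups.
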MaterialFunction /Game/Packages/Fortress_SharedMaterials/Base_Material_Functions/BaseCustomizationFunctions/MF_FN_Dummy_MASTER
graph {nodes [
  n1 = MaterialFunctionCall() {materialFunction=&/MF_Base_CroppedEmissive}
  n2 = TextureSample() {samplerSource=SSM_Wrap_WorldGroupSettings, samplerType=SAMPLERTYPE_Normal, texture=&FlatNormalBlue}
  FunctionOutput(None=MakeMaterialAttributes!(
    BaseColor
    Metallic
    Specular
    Roughness
    Anisotropy
    EmissiveColor=@n1.Result
    Opacity
    OpacityMask
    Normal=@n2.RGB
    Tangent
    WorldPositionOffset
    SubsurfaceColor
    ClearCoat
    ClearCoatRoughness
    AmbientOcclusion
    Refraction
    CustomizedUVs_0
    CustomizedUVs_1
    CustomizedUVs_2
    CustomizedUVs_3
    CustomizedUVs_4
    CustomizedUVs_5
    CustomizedUVs_6
    CustomizedUVs_7
    PixelDepthOffset
    ShadingModel
    Displacement
   )) {outputName=Material Result, sortPriority=-200}
  FunctionOutput(None=Constant()) {outputName=-------------------, sortPriority=23}
  n3 = MaterialFunctionCall!(UV 1 Coordinates) {materialFunction=&/MF_Base_CustomizationMask}
  FunctionOutput(None=@n3."Customization Mask 1 (_CM R)") {outputName="Customization Mask 1 (_CM B Top Left)", sortPriority=24}
  FunctionOutput(None=@n3."Customization Mask 2 (_CM B)") {outputName="Customization Mask 2 (_CM B Top Right)", sortPriority=25}
  n4 = MaterialFunctionCall() {materialFunction=&/MF_Base_LocalPosition}
  FunctionOutput(None=@n4.Pre-Skinned Local Normal Pixel Shader) {outputName=Pre-Skinned Local Normal Pixel Shader, sortPriority=42}
  FunctionOutput(None=@n3.Customization Mask UV Islands) {outputName="Customization Mask UV Islands (_CM B Bottom Right)", sortPriority=22}
  FunctionOutput(None=Constant()) {outputName=------------------, sortPriority=-19}
  FunctionOutput(None=@n3."Customization Crevasse (_CM G)") {outputName="Customization Crevasse (_CM G)", sortPriority=21}
  FunctionOutput(None=@n1.Cropped Emissive Texture) {description=Passthrough for Customization, outputName=Cropped Emissive Mask, sortPriority=20}
  FunctionOutput(None=Constant()) {outputName=--------------------, sortPriority=40}
  FunctionOutput(None=@n4.Pre-Skinned Local Position Pixel Shader) {outputName=Pre-Skinned Local Position Pixel Shader, sortPriority=41}
  FunctionOutput(None=@n3."Customization Mask 3 (_CM G)") {outputName="Customization Mask 3 (_CM B Bottom Left)", sortPriority=26}
  FunctionOutput(None=@n2.RGB) {description=Passthrough for Customization, outputName=Normal Map, sortPriority=15}
  FunctionOutput(None=@n4.Pre-Skinned Local Bounds Pixel Shader) {outputName=Pre-Skinned Local Bounds Pixel Shader, sortPriority=43}
  FunctionOutput(None=@n3."Customization GMAP (_CM A)") {outputName=Customization Gmap, sortPriority=21}
 ]}
