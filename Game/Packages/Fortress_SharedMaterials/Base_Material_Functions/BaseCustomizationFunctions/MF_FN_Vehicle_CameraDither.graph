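# Material function for vehicle camera fading that lowers Opacity and OpacityMask on pixels near the camera around screen center.
# An aspect-corrected, skewable screen-space radial gradient drops opacity toward MinimumOpacity inside the circle.
# A hard world-space SphereMask around the camera limits the effect, its radius blending between normal and ADS values via PlayerIsADS.
# The DitherTemporalAA call and several screen-ratio nodes do not feed the output.
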MaterialFunction /Game/Packages/Fortress_SharedMaterials/Base_Material_Functions/BaseCustomizationFunctions/MF_FN_Vehicle_CameraDither
graph {nodes [
  n1 = MaterialFunctionCall(Float2=ViewSize()) {materialFunction=&BreakOutFloat2Components}
  Divide(A=@n1.R, B=@n1.R)
  Divide(A=@n1.G, B=@n1.G)
  n2 = GetMaterialAttributes(None=FunctionInput() {inputType=FunctionInput_MaterialAttributes}) {attributeGetTypes [#1, #2]}
  n3 = MaterialFunctionCall(Float2=MaterialFunctionCall() {materialFunction=&ScreenResolution}.Visible Resolution) {materialFunction=&BreakOutFloat2Components}
  n4 = Divide(A=@n3.G, B=@n3.G)
  n5 = Divide(B=@n4)
  n6 = ScalarParameter() {defaultValue=0.66, parameterName=WidthSkew}
  n7 = MaterialFunctionCall(
   UVs=ScreenPosition().ViewportUV
   Texture Scale=AppendVector(A=ScalarParameter() {defaultValue=1, parameterName=HeightSkew}, B=Multiply(A=@n5, B=@n6))
  ) {materialFunction=&ScaleUVsByCenter}
  n8 = VectorParameter() {parameterName=ScreenCenterPosition, defaultValue=<rgba 0.5 0.5 0 1>}
  n9 = AppendVector(A=@n8.R, B=@n8.R)
  n10 = ScalarParameter() {defaultValue=0.33, group=Vehicle Camera Dithering, parameterName=Circle Radius}
  n11 = LinearInterpolate(
   A=ScalarParameter() {defaultValue=-0.5, group=Vehicle Camera Dithering, parameterName=MinimumOpacity}
   Alpha=OneMinus(None=MaterialFunctionCall!(UVs=@n7.UVs, CenterPosition=@n9, Radius=@n10) {materialFunction=/Engine/Functions/Engine_MaterialFunctions01/Gradient/SphereGradient-2D}.Result 0-1)
  ) {constA=0.25, constB=1.1}
  n12 = CameraPositionWS()
  n13 = WorldPosition()
  n14 = LinearInterpolate(
   A=ScalarParameter() {defaultValue=200, group=Vehicle Camera Dithering, parameterName=LegalRad}
   B=ScalarParameter() {defaultValue=50, group=Vehicle Camera Dithering, parameterName=LegalRadADS}
   Alpha=ScalarParameter() {parameterName=PlayerIsADS}
  )
  n15 = Constant() {r=1}
  n16 = Clamp(None=LinearInterpolate(A=@n2.OpacityMask, B=@n11, Alpha=SphereMask(A=@n12, B=@n13.XYZ, Radius=@n14, Hardness=@n15)))
  FunctionOutput(None=SetMaterialAttributes!(MaterialAttributes=@n2.MaterialAttributes, OpacityMask=Multiply(A=@n2.OpacityMask, B=@n16), Opacity=Multiply(A=@n2.Opacity, B=@n16)) {attributeSetTypes [#1, #2]})
  MaterialFunctionCall(Alpha Threshold=@n11, Random=Constant() {r=1}, Offset2=Constant() {r=1}) {materialFunction=&DitherTemporalAA_Fortnite}
  n17 = MaterialFunctionCall() {materialFunction=&BreakOutFloat2Components}
  Multiply(A=@n4, B=Divide(A=@n17.R, B=@n17.R) {desc=ratio})
  Multiply(
   A=MaterialFunctionCall(Texture Dimensions=VectorParameter() {parameterName=Tweak_Dimension, defaultValue=<rgba 1024 1024 0 1>}.RGB) {materialFunction=&ScreenAlignedUVs}."X 100%, Y 100%"
   B=AppendVector()
  )
  ScalarParameter() {defaultValue=100, parameterName=x Scale}
  ScalarParameter() {defaultValue=100, parameterName=y Scale}
 ]}
