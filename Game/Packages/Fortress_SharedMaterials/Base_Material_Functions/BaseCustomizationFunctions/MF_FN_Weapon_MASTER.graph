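# Material function that serves as the Fortnite weapon master, wrapping M_GunMasterShaderFunction to expose weapon shading and customization data.
# It passes through the material result, diffuse, normal map, cropped emissive mask and customization masks, crevasse and gmap from that function.
# Pre-skinned local position also comes from the gun function, while MF_Base_LocalPosition supplies pre-skinned local normal and bounds.
# Constant-fed dashed outputs act as separators between output groups.
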MaterialFunction /Game/Packages/Fortress_SharedMaterials/Base_Material_Functions/BaseCustomizationFunctions/MF_FN_Weapon_MASTER
graph {nodes [
  FunctionInput()
  n1 = MaterialFunctionCall!(UV 1 coordinates, MuzzleFalloff) {materialFunction=/Game/Packages/Fortnite_UniqueMaterials_1/Weapons/M_GunMasterShaderFunction}
  FunctionOutput(None=@n1.Material Result) {outputName=Material Result, sortPriority=-200}
  FunctionOutput(None=Constant()) {outputName=-------------------, sortPriority=23}
  FunctionOutput(None=@n1."Customization Mask 1 (_CM B Top Left)") {outputName="Customization Mask 1 (_CM B Top Left)", sortPriority=24}
  FunctionOutput(None=@n1."Customization Mask 2 (_CM B Top Right)") {outputName="Customization Mask 2 (_CM B Top Right)", sortPriority=25}
  FunctionOutput(None=@n1.Normal Texture) {description=Passthrough for Customization, outputName=Normal Map, sortPriority=15}
  FunctionOutput(None=@n1."Customization Mask UV Islands (_CM B Bottom Right)") {outputName="Customization Mask UV Islands (_CM B Bottom Right)", sortPriority=22}
  FunctionOutput(None=Constant()) {outputName=------------------, sortPriority=-19}
  FunctionOutput(None=@n1."Customization Crevasse (_CM G)") {outputName="Customization Crevasse (_CM G)", sortPriority=21}
  FunctionOutput(None=@n1.Cropped Emissive Mask) {description=Passthrough for Customization, outputName=Cropped Emissive Mask, sortPriority=20}
  FunctionOutput(None=Constant()) {outputName=--------------------, sortPriority=40}
  FunctionOutput(None=@n1."Customization Mask (_CM RGBA)") {outputName="Customization Mask (_CM RGBA)", sortPriority=31}
  FunctionOutput(None=@n1.Pre-Skinned Local Position Pixel Shader) {outputName=Pre-Skinned Local Position Pixel Shader, sortPriority=41}
  FunctionOutput(None=@n1."Customization Mask 3 (_CM B Bottom Left)") {outputName="Customization Mask 3 (_CM B Bottom Left)", sortPriority=26}
  n2 = MaterialFunctionCall() {materialFunction=&/MF_Base_LocalPosition}
  FunctionOutput(None=@n2.Pre-Skinned Local Bounds Pixel Shader) {outputName=Pre-Skinned Local Bounds Pixel Shader, sortPriority=43}
  FunctionOutput(None=@n2.Pre-Skinned Local Normal Pixel Shader) {outputName=Pre-Skinned Local Normal Pixel Shader, sortPriority=42}
  FunctionOutput(None=@n1.Customization Gmap) {outputName=Customization Gmap, sortPriority=21}
  FunctionOutput(None=@n1.Diffuse) {outputName=Diffuse}
 ]}
