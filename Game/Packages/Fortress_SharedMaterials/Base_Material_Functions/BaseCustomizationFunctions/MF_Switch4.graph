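# MF_Switch4 is a material function in the base customization set that switches four paired inputs at once.
# A single static bool input, Alpha, drives four static switches, each sending its A input to the matching output when true and its B input when false.
# The choice resolves at shader compile time, so one flag swaps all four channels without runtime blending.
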MaterialFunction /Game/Packages/Fortress_SharedMaterials/Base_Material_Functions/BaseCustomizationFunctions/MF_Switch4
graph {nodes [
  n1 = FunctionInput() {inputName=Alpha, inputType=FunctionInput_StaticBool, sortPriority=8}
  FunctionOutput(None=StaticSwitch(True=FunctionInput() {inputName=A_01, sortPriority=-1}, False=FunctionInput() {inputName=B_01}, Value=@n1)) {outputName=Out_01}
  FunctionOutput(None=StaticSwitch(True=FunctionInput() {inputName=A_02, sortPriority=1}, False=FunctionInput() {inputName=B_02, sortPriority=2}, Value=@n1)) {outputName=Out_02, sortPriority=2}
  FunctionOutput(None=StaticSwitch(True=FunctionInput() {inputName=A_03, sortPriority=3}, False=FunctionInput() {inputName=B_03, sortPriority=4}, Value=@n1)) {outputName=Out_03, sortPriority=3}
  FunctionOutput(None=StaticSwitch(True=FunctionInput() {inputName=A_04, sortPriority=5}, False=FunctionInput() {inputName=B_04, sortPriority=5}, Value=@n1)) {outputName=Out_04, sortPriority=4}
 ]}
