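# Material function that builds an animated panning mask for diffuse customization on weapons, vehicles and mechs, wrapping a gradient texture along normalized local position.
# Static switches choose single-column or full-texture UVs, with axes and offsets blended by isVehicle and isMech, then pan continuously or by sine.
# DiffusMask green and blue warp the UVs, and HeightLerp or Power shapes the output.
# A constant lerp alpha of 1 bypasses the scale/step-animation branch.
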MaterialFunction /Game/Packages/Fortress_SharedMaterials/Base_Material_Functions/BaseCustomizationFunctions/MF_WrapDiffusePan
graph {nodes [
  Constant()
  n1 = StaticSwitchParameter(
   True=Constant3Vector() {constant=<rgba 1 1 1 0>}
   False=FunctionInput(Preview=Constant3Vector() {constant=<rgba 0 1 0 0>}) {bUsePreviewValueAsDefault=true, inputName=ForwardDirection}
  ) {
   desc=uses the full texter vs single pixel collum
   group=Primary Triplaner DiffusePanningMask
   parameterName=Tri-01-PanningMaskFullTexture
   sortPriority=-3
  }
  n2 = MaterialFunctionCall(Float3=Multiply(A=FunctionInput() {inputName=NormalizedLocalPosition}, B=@n1)) {materialFunction=&BreakOutFloat3Components}
  n3 = FunctionInput() {inputName=isVehicle}
  n4 = LinearInterpolate(
   A=AppendVector(A=@n2.G, B=@n2.G)
   B=StaticSwitchParameter(True=AppendVector(A=@n2.G, B=@n2.G), False=AppendVector(A=@n2.R, B=@n2.R)) {group=Primary Triplaner DiffusePanningMask, parameterName=Tri01pan_Rotate_Vehicle_direction, sortPriority=5}
   Alpha=@n3
  )
  n5 = StaticSwitchParameter(True=AppendVector(A=@n2.R, B=@n2.R), False=AppendVector(A=@n2.B, B=@n2.B)) {group=Primary Triplaner DiffusePanningMask, parameterName=Tri01pan_Rotate_Mech_direction, sortPriority=5}
  n6 = FunctionInput() {inputName=isMech, inputType=FunctionInput_Scalar}
  n7 = Max(A=@n2.R, B=@n2.R)
  n8 = ScalarParameter() {group=Primary Triplaner DiffusePanningMask, parameterName=Tri01 PanningMaskWeaponOffset, sortPriority=6}
  n9 = ScalarParameter() {group=Primary Triplaner DiffusePanningMask, parameterName=Tri01 PanningMaskVehicleOffset, sortPriority=7}
  n10 = LinearInterpolate(
   A=AppendVector(A=@n8, B=ScalarParameter() {group=Primary Triplaner DiffusePanningMask, parameterName=Tri01 PanningMaskWeaponOffsetX, sortPriority=6})
   B=AppendVector(A=@n9, B=ScalarParameter() {group=Primary Triplaner DiffusePanningMask, parameterName=Tri01 PanningMaskVehicleOffsetX, sortPriority=7})
   Alpha=@n3
  )
  n11 = ScalarParameter() {group=Primary Triplaner DiffusePanningMask, parameterName=Tri01 PanningMaskMechOffset, sortPriority=7}
  n12 = AppendVector(A=@n11, B=ScalarParameter() {group=Primary Triplaner DiffusePanningMask, parameterName=Tri01 PanningMaskMechOffsetX, sortPriority=7})
  n13 = LinearInterpolate(A=@n8, B=@n9, Alpha=@n3)
  n14 = Add(
   A=StaticSwitchParameter(True=LinearInterpolate(A=@n4, B=@n5, Alpha=@n6), False=Max(A=@n7, B=@n2.B)) {group=Primary Triplaner DiffusePanningMask, parameterName=Tri-01-PanningMaskFullTexture, sortPriority=-3}
   B=StaticSwitchParameter(True=LinearInterpolate(A=@n10, B=@n12, Alpha=@n6), False=LinearInterpolate(A=@n13, B=@n11, Alpha=@n6)) {group=Primary Triplaner DiffusePanningMask, parameterName=Tri-01-PanningMaskFullTexture, sortPriority=-3}
  )
  n15 = Constant() {r=1}
  n16 = ScalarParameter() {defaultValue=1, group=Primary Triplaner DiffusePanningMask, parameterName=Tri01 PanningMaskPanSpeed_Y, sortPriority=22}
  n17 = ScalarParameter() {group=Primary Triplaner DiffusePanningMask, parameterName=Tri01 PanningMaskPanSpeed_X, sortPriority=21}
  n18 = Multiply(
   A=Time()
   B=StaticSwitchParameter(True=AppendVector(A=@n16, B=@n17), False=@n16) {group=Primary Triplaner DiffusePanningMask, parameterName=Tri-01-PanningMaskFullTexture, sortPriority=-3}
  )
  n19 = Divide(A=Add(A=Sine(None=@n18)))
  n20 = VectorParameter() {
   group=Primary Triplaner DiffusePanningMask
   parameterName=Tri01 PanningMaskStepJumpAmountPerStep
   sortPriority=27
   defaultValue=<rgba 1 1 1 0>
  }
  n21 = ComponentMask(None=@n20.RGB) {g=true, r=true}
  n22 = Ceil(None=Multiply(A=@n19, B=@n21) {constB=30})
  n23 = LinearInterpolate(
   A=AppendVector(
    A=ScalarParameter() {defaultValue=1, group=Primary Triplaner DiffusePanningMask, parameterName=Tri01 PanningMaskSpreadMax, sortPriority=23}
    B=@n15
   )
   B=AppendVector(
    A=ScalarParameter() {defaultValue=1, group=Primary Triplaner DiffusePanningMask, parameterName=Tri01 PanningMaskSpreadMin, sortPriority=24}
    B=@n15
   )
   Alpha=StaticSwitchParameter(True=Divide(A=@n22, B=@n21), False=@n19) {group=Primary Triplaner DiffusePanningMask, parameterName=WrapDiffusePan_StepAnimation, sortPriority=26}
  )
  n24 = ScalarParameter() {defaultValue=1, group=Primary Triplaner DiffusePanningMask, parameterName=Tri01 PanningMaskSpreadX}
  n25 = ScalarParameter() {defaultValue=1, group=Primary Triplaner DiffusePanningMask, parameterName=Tri01 PanningMaskSpreadY, sortPriority=1}
  n26 = Multiply(
   A=StaticSwitchParameter(
    True=AppendVector(A=@n24, B=@n25)
    False=ScalarParameter() {defaultValue=1, group=Primary Triplaner DiffusePanningMask, parameterName=Tri01 PanningMaskSpread}
   ) {group=Primary Triplaner DiffusePanningMask, parameterName=Tri-01-PanningMaskFullTexture, sortPriority=-2}
   B=@n14
  )
  n27 = Add(A=@n26, B=@n18)
  n28 = StaticSwitchParameter(True=Multiply(A=@n14, B=@n23), False=@n27) {group=Primary Triplaner DiffusePanningMask, parameterName=Tri-01-PanningMask_Continuous/Scale, sortPriority=20}
  n29 = ScalarParameter() {defaultValue=1, group=Primary Triplaner DiffusePanningMask, parameterName=Tri01 PanningMaskSineMax, sortPriority=23}
  n30 = Constant() {r=1}
  n31 = ScalarParameter() {defaultValue=1, group=Primary Triplaner DiffusePanningMask, parameterName=Tri01 PanningMaskSineMin, sortPriority=24}
  n32 = Subtract(A=Sine(None=@n18))
  n33 = Add(A=@n14, B=LinearInterpolate(A=AppendVector(A=@n29, B=@n30), B=AppendVector(A=@n31, B=@n30), Alpha=Divide(A=@n32)))
  n34 = StaticSwitchParameter(True=Multiply(A=@n26, B=@n33), False=@n27) {group=Primary Triplaner DiffusePanningMask, parameterName=Tri-01-PanningMask_Continuous/Sine, sortPriority=20}
  n35 = Constant() {r=1}
  n36 = MaterialFunctionCall(Color=FunctionInput() {bUsePreviewValueAsDefault=true, inputName=DiffusMask}) {materialFunction=&SplitComponents}
  n37 = ScalarParameter() {
   defaultValue=0.1
   desc="Warp uv based off Blue channel of mask "
   group=Primary Triplaner DiffusePanningMask
   parameterName="Tri01 PanMaskWarpUVs (Blue Channel)"
   sortPriority=8
  }
  n38 = Multiply(A=@n36.B, B=@n37)
  n39 = ScalarParameter() {
   defaultValue=0.1
   desc="Warp uv based off Green channel of mask "
   group=Primary Triplaner DiffusePanningMask
   parameterName="Tri01 PanMaskWarpUvs (Green Channel)"
   sortPriority=10
  }
  n40 = StaticSwitchParameter(True=Constant() {r=1}, False=@n36.G) {
   desc=Fill green chanel with pure white to use blue as warping uvs used with rgb is a grayscale image
   group=Primary Triplaner DiffusePanningMask
   parameterName=Tri01 PanWarpUvsFillGreenChannelWhite
   sortPriority=9
  }
  n41 = Multiply(A=@n39, B=@n40)
  n42 = TextureObjectParameter() {
   group=Primary Triplaner DiffusePanningMask
   parameterName=WrapDiffusGradient
   samplerType=SAMPLERTYPE_Masks
   sortPriority=5
   texture=/Game/Global/Textures/Default/Masks_and_Gradients/T_VertGradient_MASK
  }
  n43 = TextureSample(UVs=Add(A=LinearInterpolate(A=@n28, B=@n34, Alpha=@n35), B=Add(A=@n38, B=@n41)), Tex=@n42) {samplerSource=SSM_Wrap_WorldGroupSettings, texture=/Game/Weapons/WeaponSkins/Textures/T_Swirly_Noise_highContrast}
  n44 = ScalarParameter() {defaultValue=5, group=Primary Triplaner DiffusePanningMask, parameterName=Tri01 PanningMaskContrast, sortPriority=12}
  n45 = Add(A=@n43.R, B=Multiply(A=@n43.R, B=@n40))
  n46 = StaticSwitchParameter(
   True=MaterialFunctionCall(Transition Phase=@n43.R, Height Texture=@n40, Contrast=@n44) {materialFunction=&HeightLerp}.Alpha
   False=Power(Base=Multiply(A=@n45, B=@n40), Exp=@n44)
  ) {group=Primary Triplaner DiffusePanningMask, parameterName=Tri-01-PanningMask-Gradient/Height, sortPriority=11}
  n47 = ScalarParameter() {defaultValue=1, group=Primary Triplaner DiffusePanningMask, parameterName=Tri01 PanningMaskOpacity, sortPriority=-4}
  FunctionOutput(None=Saturate(None=Multiply(A=@n46, B=@n47)))
  Multiply(A=@n33, B=@n23)
 ]}
