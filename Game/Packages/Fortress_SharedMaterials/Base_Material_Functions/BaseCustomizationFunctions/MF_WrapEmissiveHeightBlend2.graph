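# Material function that wraps an emissive height-gradient glow along the pre-skinned local bounds of weapons, vehicles, mechs, characters and backpacks.
# Normalized position, with per-type axes and offsets, becomes gradient UVs, optionally animated by panning, bouncing or stepped scaling.
# HeightLerp blends that gradient with EmissiveMask, yielding tinted or texture-colored emissive plus a separately contrasted Mask.
# The blue-channel UV warp is computed but left unconnected.
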MaterialFunction /Game/Packages/Fortress_SharedMaterials/Base_Material_Functions/BaseCustomizationFunctions/MF_WrapEmissiveHeightBlend2
graph {nodes [
  n1 = PreSkinnedPosition()
  n2 = VertexInterpolator(VS=PreSkinnedLocalBounds().Half Extents)
  n3 = Divide(A=Add(A=VertexInterpolator(VS=@n1).PS, B=@n2.PS), B=@n2.PS)
  n4 = Constant3Vector() {constant=<rgba 1 1 1 0>}
  n5 = FunctionInput(Preview=Constant3Vector() {constant=<rgba 0 1 0 0>}) {inputName=ForwardDirection}
  n6 = MaterialFunctionCall(Float3=Multiply(
    A=FunctionInput(Preview=@n3) {inputName=NormalizedLocalPosition}
    B=StaticSwitchParameter(True=@n4, False=@n5) {group=Primary Triplaner Emissive, parameterName=Tri 02 EmissiveHieghtUseFullGradent, sortPriority=51}
   )) {materialFunction=&BreakOutFloat3Components}
  n7 = AppendVector(A=@n6.G, B=@n6.G)
  n8 = StaticSwitchParameter(True=AppendVector(A=@n6.G, B=@n6.G), False=AppendVector(A=@n6.R, B=@n6.R)) {group=Primary Triplaner Emissive, parameterName=Tri02_Rotate_Vehicle_direction, sortPriority=55}
  n9 = FunctionInput(Preview=Constant()) {inputName=isVehicle}
  n10 = AppendVector(A=@n6.R, B=@n6.R)
  n11 = AppendVector(A=@n6.B, B=@n6.B)
  n12 = FunctionInput() {bUsePreviewValueAsDefault=true, inputName=IsMech, inputType=FunctionInput_Scalar}
  n13 = StaticSwitchParameter(
   True=LinearInterpolate(
    A=LinearInterpolate(A=@n7, B=@n8, Alpha=@n9)
    B=StaticSwitchParameter(True=@n10, False=@n11) {group=Primary Triplaner Emissive, parameterName=Tri02_Rotate_Mech_direction, sortPriority=55}
    Alpha=@n12
   )
   False=Max(A=Max(A=@n6.R, B=@n6.R), B=@n6.B)
  ) {
   desc=uses the full texture vs single pixel column
   group=Primary Triplaner Emissive
   parameterName=Tri 02 EmissiveHieghtUseFullGradent
   sortPriority=51
  }
  n14 = ScalarParameter() {defaultValue=0.65, group=Primary Triplaner Emissive, parameterName=Tri02EmissiveHieghtWeaponOffset, sortPriority=55}
  n15 = Constant()
  n16 = ScalarParameter() {defaultValue=0.65, group=Primary Triplaner Emissive, parameterName=Tri02EmissiveHieghtVehicleOffset, sortPriority=56}
  n17 = LinearInterpolate(A=AppendVector(A=@n14, B=@n15), B=AppendVector(A=@n16, B=@n15), Alpha=@n9)
  n18 = LinearInterpolate(
   A=ScalarParameter() {defaultValue=0.65, group=Primary Triplaner Emissive, parameterName=Tri02EmissiveHieghtMechOffset, sortPriority=56}
   B=LinearInterpolate(
    A=ScalarParameter() {defaultValue=0.05, group=Primary Triplaner Emissive, parameterName=Tri02EmissiveHieghtCharacterOffset, sortPriority=56}
    B=ScalarParameter() {defaultValue=0.05, group=Primary Triplaner Emissive, parameterName=Tri02EmissiveHieghtBackpackOffset, sortPriority=56}
    Alpha=NamedRerouteUsage()
   )
   Alpha=NamedRerouteUsage()
  )
  n19 = AppendVector(A=@n18, B=@n15)
  n20 = LinearInterpolate(A=@n14, B=@n16, Alpha=@n9)
  n21 = Add(
   A=@n13
   B=StaticSwitchParameter(True=LinearInterpolate(A=@n17, B=@n19, Alpha=@n12), False=LinearInterpolate(A=@n20, B=@n18, Alpha=@n12)) {group=Primary Triplaner Emissive, parameterName=Tri 02 EmissiveHieghtUseFullGradent, sortPriority=51}
  )
  n22 = Constant() {r=1}
  n23 = ScalarParameter() {defaultValue=0.2, group=Primary Triplaner Emissive, parameterName=Tri02EmissiveHeightPanSpeed, sortPriority=72}
  n24 = ScalarParameter() {group=Primary Triplaner Emissive, parameterName=Tri02EmissiveHeightPanSpeed X, sortPriority=73}
  n25 = Multiply(
   A=Time()
   B=StaticSwitchParameter(True=AppendVector(A=@n23, B=@n24), False=@n23) {group=Primary Triplaner Emissive, parameterName=Tri 02 EmissiveHieghtUseFullGradent, sortPriority=51}
  )
  n26 = Divide(A=Add(A=Sine(None=@n25)))
  n27 = VectorParameter() {group=Primary Triplaner Emissive, parameterName=Tri02EmissiveStepJumpAmountPerStep, sortPriority=81, defaultValue=<rgba 1 1 1 0>}
  n28 = ComponentMask(None=@n27.RGB) {g=true, r=true}
  n29 = Ceil(None=Multiply(A=@n26, B=@n28) {constB=30})
  n30 = LinearInterpolate(
   A=AppendVector(A=ScalarParameter() {defaultValue=1, group=Primary Triplaner Emissive, parameterName=Tri02EmissiveHieghtSpreadMax, sortPriority=76}, B=@n22)
   B=AppendVector(A=ScalarParameter() {defaultValue=1, group=Primary Triplaner Emissive, parameterName=Tri02EmissiveHieghtSpreadMin, sortPriority=77}, B=@n22)
   Alpha=StaticSwitchParameter(True=Divide(A=@n29, B=@n28), False=@n26) {group=Primary Triplaner Emissive, parameterName=Tri02EmissivePan_StepAnimation, sortPriority=80}
  )
  n31 = ScalarParameter() {defaultValue=1, group=Primary Triplaner Emissive, parameterName=Tri02EmissiveHieghtSpreadX, sortPriority=53}
  n32 = ScalarParameter() {defaultValue=1, group=Primary Triplaner Emissive, parameterName=Tri02EmissiveHieghtSpreadY, sortPriority=54}
  n33 = Multiply(
   A=StaticSwitchParameter(
    True=AppendVector(A=@n31, B=@n32)
    False=ScalarParameter() {defaultValue=0.25, group=Primary Triplaner Emissive, parameterName=Tri02EmissiveHieghtSpread, sortPriority=53}
   ) {group=Primary Triplaner Emissive, parameterName=Tri 02 EmissiveHieghtUseFullGradent, sortPriority=51}
   B=@n21
  )
  n34 = Add(A=Sine(None=ComponentMask(None=@n25) {r=true}))
  n35 = StaticSwitchParameter(True=LinearInterpolate(Alpha=Divide(A=@n34)) {constA=0.65, constB=1.3}, False=@n25) {group=Primary Triplaner Emissive, parameterName=Tri02EmissiveHieghtPan_Continuous/Bounce, sortPriority=71}
  n36 = StaticSwitchParameter(
   True=StaticSwitchParameter(True=Multiply(A=@n21, B=@n30), False=Add(A=@n33, B=@n35)) {group=Primary Triplaner Emissive, parameterName=Tri02EmissiveHieghtPan_Continuous/Scale2, sortPriority=75}
   False=@n33
  ) {group=Primary Triplaner Emissive, parameterName=Tri02-Animate-Emissive-Height, sortPriority=70}
  n37 = ScalarParameter() {desc="Warp Uvs using the Blue channel ", group=Primary Triplaner Emissive, parameterName=Tri02EmissiveHeightWarpUvs2, sortPriority=58}
  Add(A=@n36, B=Multiply(A=FunctionInput() {inputName=EmissiveMask_B, inputType=FunctionInput_Scalar, sortPriority=10}, B=@n37))
  n38 = TextureObjectParameter() {
   group=Primary Triplaner Emissive
   parameterName=WrapEmissiveHeightGradient2
   samplerType=SAMPLERTYPE_Masks
   sortPriority=52
   texture=/Game/Global/Textures/Default/Masks_and_Gradients/T_VertGradient_MASK
  }
  n39 = TextureSample(UVs=@n36, Tex=@n38) {texture=/Game/Weapons/WeaponSkins/Textures/T_Swirly_Noise_highContrast}
  n40 = FunctionInput() {inputName=EmissiveMask, inputType=FunctionInput_Scalar}
  n41 = MaterialFunctionCall(
   Transition Phase=@n39.R
   Height Texture=@n40
   Contrast=ScalarParameter() {defaultValue=1, group=Primary Triplaner Emissive, parameterName=Tri02EmissiveHeightContrast, sortPriority=57}
  ) {materialFunction=&HeightLerp}
  n42 = TextureObjectParameter() {
   desc=Alpha is use as a mask for the intenisty
   group=Primary Triplaner Emissive
   parameterName=Tri02EmissiveHeightGradientColor2
   sortPriority=61
   texture=/Game/Global/Textures/Default/Thinfilm/T_ThinFilm_Spectrum_DESATURATED-50
  }
  n43 = TextureSample(
   UVs=StaticSwitchParameter(True=@n41.Alpha, False=@n36) {group=Primary Triplaner Emissive, parameterName="Tri02EmissiveHeightColor  Uv/Height", sortPriority=60}
   Tex=@n42
  ) {texture=&T_Swirly_Noise_highContrast}
  n44 = ScalarParameter() {
   defaultValue=1
   desc=Alpha is use as a mask for the intenisty
   group=Primary Triplaner Emissive
   parameterName=Tri02EmissiveHLColor_Intensty
   sortPriority=62
  }
  n45 = LinearInterpolate(A=@n43.RGB, B=Multiply(A=@n43.RGB, B=@n44), Alpha=@n43.RGB)
  n46 = VectorParameter() {group=Primary Triplaner Emissive, parameterName=Tri02EmissiveHeightColor, sortPriority=60, defaultValue=<rgba 0 0 0 1>}
  FunctionOutput(None=Multiply(
    A=@n41.Alpha
    B=StaticSwitchParameter(True=@n45, False=@n46.RGB) {group=Primary Triplaner Emissive, parameterName=Tri02-Emissive-Height-Color-Texture, sortPriority=59}
   ))
  FunctionOutput(None=MaterialFunctionCall(
    Transition Phase=@n39.R
    Height Texture=@n40
    Contrast=ScalarParameter() {defaultValue=1, group=Primary Triplaner Emissive, parameterName=Tri02EmissiveHeightMaskContrast, sortPriority=20}
   ) {materialFunction=&HeightLerp}.Alpha) {outputName=Mask}
  NamedRerouteDeclaration(None=FunctionInput() {bUsePreviewValueAsDefault=true, inputName=IsCharacter, inputType=FunctionInput_Scalar}) {name=IsCharacter, nodeColor=<rgba 0 0.9253 1 1>}
  n47 = FunctionInput() {bUsePreviewValueAsDefault=true, inputName=IsStatsB, inputType=FunctionInput_Vector2}
  NamedRerouteDeclaration(None=ComponentMask(None=@n47) {r=true}) {name=IsBackpack, nodeColor=<rgba 0 0.9253 1 1>}
  NamedRerouteDeclaration(None=ComponentMask(None=@n47) {g=true}) {name=IsPickaxe, nodeColor=<rgba 0 0.9253 1 1>}
 ]}
