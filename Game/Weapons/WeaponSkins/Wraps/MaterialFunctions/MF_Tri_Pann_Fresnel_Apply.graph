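# Material function for weapon wrap skins that applies a fresnel tint to incoming base color and emissive maps.
# A fresnel term from vertex normal and camera vector, shaped by multiply and power parameters, blends dark and light colors that multiply the base color.
# The same term, colored and scaled by an emissive amount, is added to emissive.
# One static switch, PanMaskFresnel_Switch, toggles both effects, passing inputs through unchanged when off.
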MaterialFunction /Game/Weapons/WeaponSkins/Wraps/MaterialFunctions/MF_Tri_Pann_Fresnel_Apply
graph {nodes [
  n1 = VectorParameter() {group=Primary Triplaner DiffusePanningMask, parameterName=Tri_01_DarkFresnelColor, defaultValue=<rgba 0.0103 0.0114 0.1 1>}
  n2 = VertexNormalWS()
  n3 = CameraVectorWS()
  n4 = OneMinus(None=Saturate(None=DotProduct(A=@n2, B=@n3)))
  n5 = ScalarParameter() {defaultValue=1, group=Primary Triplaner DiffusePanningMask, parameterName=FresnelMultiply}
  n6 = Saturate(None=Power(Base=Multiply(A=@n4, B=@n5), Exp=ScalarParameter() {defaultValue=2, group=Primary Triplaner DiffusePanningMask, parameterName=FresnelPower}))
  n7 = ScalarParameter() {defaultValue=0.3, group=Primary Triplaner DiffusePanningMask, parameterName=FresnelDiffuseAmount}
  n8 = VectorParameter() {group=Primary Triplaner DiffusePanningMask, parameterName=Tri_01_LightFresnelColor, defaultValue=<rgba 0.0901 0.1447 0.3 1>}
  n9 = LinearInterpolate(A=@n1.RGB, B=@n8.RGB, Alpha=Saturate(None=Multiply(A=@n6, B=@n7))) {constA=1}
  n10 = FunctionInput() {bUsePreviewValueAsDefault=true, inputName=BaseColor_Map, sortPriority=15}
  FunctionOutput(None=StaticSwitchParameter(True=Multiply(A=@n9, B=@n10), False=@n10) {group=Primary Triplaner DiffusePanningMask, parameterName=PanMaskFresnel_Switch}) {outputName=Basecolor_output}
  n11 = VectorParameter() {group=Primary Triplaner DiffusePanningMask, parameterName=Tri_01_BaseFresnelColor, defaultValue=<rgba 0.2 0.5289 1 1>}
  n12 = Multiply(
   A=Multiply(A=@n6, B=@n11.RGB)
   B=ScalarParameter() {defaultValue=0.05, group=Primary Triplaner DiffusePanningMask, parameterName=FresnelEmissiveAmount}
  )
  n13 = FunctionInput() {bUsePreviewValueAsDefault=true, inputName=Emissive_Map, sortPriority=15}
  FunctionOutput(None=StaticSwitchParameter(True=Add(A=@n13, B=@n12), False=@n13) {group=Primary Triplaner DiffusePanningMask, parameterName=PanMaskFresnel_Switch}) {outputName=Emissive_output}
 ]}
